
using { /Fortnite.com/Devices }
using { /Verse.org/Simulation }
using { /UnrealEngine.com/Temporary/Diagnostics }
using { /UnrealEngine.com/Temporary/UI }
using { /UnrealEngine.com/Temporary/SpatialMath }
using { /Verse.org/Simulation }
using { /Fortnite.com/Devices }
using { /Fortnite.com/UI }
using { /Verse.org/Colors/NamedColors }
using { /Verse.org/Colors }

# See https://dev.epicgames.com/documentation/en-us/uefn/create-your-own-device-in-verse for how to create a verse device.

# A Verse-authored creative device that can be placed in a level
action_info_ui := class():

    #Class settings
    var MaybeAgent : ?agent = false
    var IsDungeonStarting : logic = false
    BakeryMachinesManager : bakery_machines_manager

    #Simon
    Simon_Slider_Texture : texture_block = texture_block{DefaultDesiredSize:=vector2{X:=XSizeActionUI,Y:=OutOfScreenY},DefaultImage:=Images.EmptyTexture}
    #Simon_Background_Texture : texture_block = texture_block{DefaultImage:=Images.BlackBackground}
    Simon_Icon_Texture : texture_block = texture_block{DefaultImage:=Images.Star, DefaultTint := Gold, DefaultDesiredSize:=vector2{X:=ActionIconSizeX,Y:=ActionIconSizeY}}
    Simon_TextBlock : text_block = text_block{DefaultTextColor:=White, DefaultShadowOffset:=ShadowOffset, DefaultShadowOpacity := 1.0}

    #Dungeon Time
    Dungeon_Slider_Texture : texture_block = texture_block{DefaultDesiredSize:=vector2{X:=XSizeActionUI,Y:=OutOfScreenY},DefaultImage:=Images.EmptyTexture}
    #Dungeon_Background_Texture : texture_block = texture_block{DefaultImage:=Images.BlackBackground}
    Dungeon_TextBlock : text_block = text_block{DefaultTextColor:=White, DefaultShadowOffset:=ShadowOffset, DefaultShadowOpacity := 1.0}
    DungeonRemainingChests_TextBlock : text_block = text_block{DefaultTextColor:=White, DefaultShadowOffset:=ShadowOffset, DefaultShadowOpacity := 1.0}

    #Slider 0
    Slider_Texture_0 : texture_block = texture_block{DefaultDesiredSize:=vector2{X:=XSizeActionUI,Y:=OutOfScreenY},DefaultImage:=Images.EmptyTexture}
    Background_Texture_0 : texture_block = texture_block{DefaultImage:=Images.BlackBackground}
    Action_Icon_0 : texture_block = texture_block{DefaultImage:=Images.EmptyTexture, DefaultDesiredSize:=vector2{X:=ActionIconSizeX,Y:=ActionIconSizeY}}
    Ammount_TextBlock_0 : text_block = text_block{DefaultTextColor:=White, DefaultShadowOffset:=ShadowOffset, DefaultShadowOpacity := 1.0}
    var Used_0 : logic = false

    #Slider 1
    Slider_Texture_1 : texture_block = texture_block{DefaultDesiredSize:=vector2{X:=XSizeActionUI,Y:=OutOfScreenY},DefaultImage:=Images.EmptyTexture}
    Background_Texture_1 : texture_block = texture_block{DefaultImage:=Images.BlackBackground}
    Action_Icon_1 : texture_block = texture_block{DefaultImage:=Images.EmptyTexture, DefaultDesiredSize:=vector2{X:=ActionIconSizeX,Y:=ActionIconSizeY}}
    Ammount_TextBlock_1 : text_block = text_block{DefaultTextColor:=White, DefaultShadowOffset:=ShadowOffset, DefaultShadowOpacity := 1.0}
    var Used_1 : logic = false

    #Slider 2
    Slider_Texture_2 : texture_block = texture_block{DefaultDesiredSize:=vector2{X:=XSizeActionUI,Y:=OutOfScreenY},DefaultImage:=Images.EmptyTexture}
    Background_Texture_2 : texture_block = texture_block{DefaultImage:=Images.BlackBackground}
    Action_Icon_2 : texture_block = texture_block{DefaultImage:=Images.EmptyTexture, DefaultDesiredSize:=vector2{X:=ActionIconSizeX,Y:=ActionIconSizeY}}
    Ammount_TextBlock_2 : text_block = text_block{DefaultTextColor:=White, DefaultShadowOffset:=ShadowOffset, DefaultShadowOpacity := 1.0}
    var Used_2 : logic = false

    #Slider 3
    Slider_Texture_3 : texture_block = texture_block{DefaultDesiredSize:=vector2{X:=XSizeActionUI,Y:=OutOfScreenY},DefaultImage:=Images.EmptyTexture}
    Background_Texture_3 : texture_block = texture_block{DefaultImage:=Images.BlackBackground}
    Action_Icon_3 : texture_block = texture_block{DefaultImage:=Images.EmptyTexture, DefaultDesiredSize:=vector2{X:=ActionIconSizeX,Y:=ActionIconSizeY}}
    Ammount_TextBlock_3 : text_block = text_block{DefaultTextColor:=White, DefaultShadowOffset:=ShadowOffset, DefaultShadowOpacity := 1.0}
    var Used_3 : logic = false

    #Slider 4
    Slider_Texture_4 : texture_block = texture_block{DefaultDesiredSize:=vector2{X:=XSizeActionUI,Y:=OutOfScreenY},DefaultImage:=Images.EmptyTexture}
    Background_Texture_4 : texture_block = texture_block{DefaultImage:=Images.BlackBackground}
    Action_Icon_4 : texture_block = texture_block{DefaultImage:=Images.EmptyTexture, DefaultDesiredSize:=vector2{X:=ActionIconSizeX,Y:=ActionIconSizeY}}
    Ammount_TextBlock_4 : text_block = text_block{DefaultTextColor:=White, DefaultShadowOffset:=ShadowOffset, DefaultShadowOpacity := 1.0}
    var Used_4 : logic = false

    #Slider 5
    Slider_Texture_5 : texture_block = texture_block{DefaultDesiredSize:=vector2{X:=XSizeActionUI,Y:=OutOfScreenY},DefaultImage:=Images.EmptyTexture}
    Background_Texture_5 : texture_block = texture_block{DefaultImage:=Images.BlackBackground}
    Action_Icon_5 : texture_block = texture_block{DefaultImage:=Images.EmptyTexture, DefaultDesiredSize:=vector2{X:=ActionIconSizeX,Y:=ActionIconSizeY}}
    Ammount_TextBlock_5 : text_block = text_block{DefaultTextColor:=White, DefaultShadowOffset:=ShadowOffset, DefaultShadowOpacity := 1.0}
    var Used_5 : logic = false

    #Slider 6
    Slider_Texture_6 : texture_block = texture_block{DefaultDesiredSize:=vector2{X:=XSizeActionUI,Y:=OutOfScreenY},DefaultImage:=Images.EmptyTexture}
    Background_Texture_6 : texture_block = texture_block{DefaultImage:=Images.BlackBackground}
    Action_Icon_6 : texture_block = texture_block{DefaultImage:=Images.EmptyTexture, DefaultDesiredSize:=vector2{X:=ActionIconSizeX,Y:=ActionIconSizeY}}
    Ammount_TextBlock_6 : text_block = text_block{DefaultTextColor:=White, DefaultShadowOffset:=ShadowOffset, DefaultShadowOpacity := 1.0}
    var Used_6 : logic = false

    #Slider 7
    Slider_Texture_7 : texture_block = texture_block{DefaultDesiredSize:=vector2{X:=XSizeActionUI,Y:=OutOfScreenY},DefaultImage:=Images.EmptyTexture}
    Background_Texture_7 : texture_block = texture_block{DefaultImage:=Images.BlackBackground}
    Action_Icon_7 : texture_block = texture_block{DefaultImage:=Images.EmptyTexture, DefaultDesiredSize:=vector2{X:=ActionIconSizeX,Y:=ActionIconSizeY}}
    Ammount_TextBlock_7 : text_block = text_block{DefaultTextColor:=White, DefaultShadowOffset:=ShadowOffset, DefaultShadowOpacity := 1.0}
    var Used_7 : logic = false

    #Slider 8
    Slider_Texture_8 : texture_block = texture_block{DefaultDesiredSize:=vector2{X:=XSizeActionUI,Y:=OutOfScreenY},DefaultImage:=Images.EmptyTexture}
    Background_Texture_8 : texture_block = texture_block{DefaultImage:=Images.BlackBackground}
    Action_Icon_8 : texture_block = texture_block{DefaultImage:=Images.EmptyTexture, DefaultDesiredSize:=vector2{X:=ActionIconSizeX,Y:=ActionIconSizeY}}
    Ammount_TextBlock_8 : text_block = text_block{DefaultTextColor:=White, DefaultShadowOffset:=ShadowOffset, DefaultShadowOpacity := 1.0}
    var Used_8 : logic = false

    #Slider 9
    Slider_Texture_9 : texture_block = texture_block{DefaultDesiredSize:=vector2{X:=XSizeActionUI,Y:=OutOfScreenY},DefaultImage:=Images.EmptyTexture}
    Background_Texture_9 : texture_block = texture_block{DefaultImage:=Images.BlackBackground}
    Action_Icon_9 : texture_block = texture_block{DefaultImage:=Images.EmptyTexture, DefaultDesiredSize:=vector2{X:=ActionIconSizeX,Y:=ActionIconSizeY}}
    Ammount_TextBlock_9 : text_block = text_block{DefaultTextColor:=White, DefaultShadowOffset:=ShadowOffset, DefaultShadowOpacity := 1.0}
    var Used_9 : logic = false

    #Slider 10
    Slider_Texture_10 : texture_block = texture_block{DefaultDesiredSize:=vector2{X:=XSizeActionUI,Y:=OutOfScreenY},DefaultImage:=Images.EmptyTexture}
    Background_Texture_10 : texture_block = texture_block{DefaultImage:=Images.BlackBackground}
    Action_Icon_10 : texture_block = texture_block{DefaultImage:=Images.EmptyTexture, DefaultDesiredSize:=vector2{X:=ActionIconSizeX,Y:=ActionIconSizeY}}
    Ammount_TextBlock_10 : text_block = text_block{DefaultTextColor:=White, DefaultShadowOffset:=ShadowOffset, DefaultShadowOpacity := 1.0}
    var Used_10 : logic = false

    #Slider 11
    Slider_Texture_11 : texture_block = texture_block{DefaultDesiredSize:=vector2{X:=XSizeActionUI,Y:=OutOfScreenY},DefaultImage:=Images.EmptyTexture}
    Background_Texture_11 : texture_block = texture_block{DefaultImage:=Images.BlackBackground}
    Action_Icon_11 : texture_block = texture_block{DefaultImage:=Images.EmptyTexture, DefaultDesiredSize:=vector2{X:=ActionIconSizeX,Y:=ActionIconSizeY}}
    Ammount_TextBlock_11 : text_block = text_block{DefaultTextColor:=White, DefaultShadowOffset:=ShadowOffset, DefaultShadowOpacity := 1.0}
    var Used_11 : logic = false

    #Slider 12
    Slider_Texture_12 : texture_block = texture_block{DefaultDesiredSize:=vector2{X:=XSizeActionUI,Y:=OutOfScreenY},DefaultImage:=Images.EmptyTexture}
    Background_Texture_12 : texture_block = texture_block{DefaultImage:=Images.BlackBackground}
    Action_Icon_12 : texture_block = texture_block{DefaultImage:=Images.EmptyTexture, DefaultDesiredSize:=vector2{X:=ActionIconSizeX,Y:=ActionIconSizeY}}
    Ammount_TextBlock_12 : text_block = text_block{DefaultTextColor:=White, DefaultShadowOffset:=ShadowOffset, DefaultShadowOpacity := 1.0}
    var Used_12 : logic = false

    #Slider 13
    Slider_Texture_13 : texture_block = texture_block{DefaultDesiredSize:=vector2{X:=XSizeActionUI,Y:=OutOfScreenY},DefaultImage:=Images.EmptyTexture}
    Background_Texture_13 : texture_block = texture_block{DefaultImage:=Images.BlackBackground}
    Action_Icon_13 : texture_block = texture_block{DefaultImage:=Images.EmptyTexture, DefaultDesiredSize:=vector2{X:=ActionIconSizeX,Y:=ActionIconSizeY}}
    Ammount_TextBlock_13 : text_block = text_block{DefaultTextColor:=White, DefaultShadowOffset:=ShadowOffset, DefaultShadowOpacity := 1.0}
    var Used_13 : logic = false

    #Slider 14
    Slider_Texture_14 : texture_block = texture_block{DefaultDesiredSize:=vector2{X:=XSizeActionUI,Y:=OutOfScreenY},DefaultImage:=Images.EmptyTexture}
    Background_Texture_14 : texture_block = texture_block{DefaultImage:=Images.BlackBackground}
    Action_Icon_14 : texture_block = texture_block{DefaultImage:=Images.EmptyTexture, DefaultDesiredSize:=vector2{X:=ActionIconSizeX,Y:=ActionIconSizeY}}
    Ammount_TextBlock_14 : text_block = text_block{DefaultTextColor:=White, DefaultShadowOffset:=ShadowOffset, DefaultShadowOpacity := 1.0}
    var Used_14 : logic = false

    #Slider 15
    Slider_Texture_15 : texture_block = texture_block{DefaultDesiredSize:=vector2{X:=XSizeActionUI,Y:=OutOfScreenY},DefaultImage:=Images.EmptyTexture}
    Background_Texture_15 : texture_block = texture_block{DefaultImage:=Images.BlackBackground}
    Action_Icon_15 : texture_block = texture_block{DefaultImage:=Images.EmptyTexture, DefaultDesiredSize:=vector2{X:=ActionIconSizeX,Y:=ActionIconSizeY}}
    Ammount_TextBlock_15 : text_block = text_block{DefaultTextColor:=White, DefaultShadowOffset:=ShadowOffset, DefaultShadowOpacity := 1.0}
    var Used_15 : logic = false

    #Slider 16
    Slider_Texture_16 : texture_block = texture_block{DefaultDesiredSize:=vector2{X:=XSizeActionUI,Y:=OutOfScreenY},DefaultImage:=Images.EmptyTexture}
    Background_Texture_16 : texture_block = texture_block{DefaultImage:=Images.BlackBackground}
    Action_Icon_16 : texture_block = texture_block{DefaultImage:=Images.EmptyTexture, DefaultDesiredSize:=vector2{X:=ActionIconSizeX,Y:=ActionIconSizeY}}
    Ammount_TextBlock_16 : text_block = text_block{DefaultTextColor:=White, DefaultShadowOffset:=ShadowOffset, DefaultShadowOpacity := 1.0}
    var Used_16 : logic = false


    #Water Resource
    WaterResource_Slider : texture_block = texture_block{DefaultDesiredSize:=vector2{X:=XSizeActionUI,Y:=OutOfScreenY},DefaultImage:=Images.EmptyTexture}
    WaterResource_Background: texture_block = texture_block{DefaultImage:=Images.BlackBackground}
    WaterResource_Icon : texture_block = texture_block{DefaultImage:=Images.Water_Icon, DefaultDesiredSize:=vector2{X:=ActionIconSizeX,Y:=ActionIconSizeY}}
    WaterResourceAmmount_TextBlock : text_block = text_block{DefaultTextColor:=WaterTextColor, DefaultShadowOffset:=ShadowOffset, DefaultShadowOpacity := 1.0}
    var WaterAmmounted : float = 0.0
    var WaterResource_DisplayTime : int = 0
    var WaterAmmountOnScreen : logic = false

    #Flour Resource
    FlourResource_Slider : texture_block = texture_block{DefaultDesiredSize:=vector2{X:=XSizeActionUI,Y:=OutOfScreenY},DefaultImage:=Images.EmptyTexture}
    FlourResource_Background: texture_block = texture_block{DefaultImage:=Images.BlackBackground}
    FlourResource_Icon : texture_block = texture_block{DefaultImage:=Images.Flour_Icon, DefaultDesiredSize:=vector2{X:=ActionIconSizeX,Y:=ActionIconSizeY}}
    FlourResourceAmmount_TextBlock : text_block = text_block{DefaultTextColor:=FlourTextColor, DefaultShadowOffset:=ShadowOffset, DefaultShadowOpacity := 1.0}
    var FlourAmmounted : float = 0.0
    var FlourResource_DisplayTime : int = 0
    var FlourAmmountOnScreen : logic = false

    #Cheese Resource
    CheeseResource_Slider : texture_block = texture_block{DefaultDesiredSize:=vector2{X:=XSizeActionUI,Y:=OutOfScreenY},DefaultImage:=Images.EmptyTexture}
    CheeseResource_Background: texture_block = texture_block{DefaultImage:=Images.BlackBackground}
    CheeseResource_Icon : texture_block = texture_block{DefaultImage:=Images.Cheese_Icon, DefaultDesiredSize:=vector2{X:=ActionIconSizeX,Y:=ActionIconSizeY}}
    CheeseResourceAmmount_TextBlock : text_block = text_block{DefaultTextColor:=CheeseTextColor, DefaultShadowOffset:=ShadowOffset, DefaultShadowOpacity := 1.0}
    var CheeseAmmounted : float = 0.0
    var CheeseResource_DisplayTime : int = 0
    var CheeseAmmountOnScreen : logic = false

    #Pimentos Resource
    PimentosResource_Slider : texture_block = texture_block{DefaultDesiredSize:=vector2{X:=XSizeActionUI,Y:=OutOfScreenY},DefaultImage:=Images.EmptyTexture}
    PimentosResource_Background: texture_block = texture_block{DefaultImage:=Images.BlackBackground}
    PimentosResource_Icon : texture_block = texture_block{DefaultImage:=Images.Pimento_Icon, DefaultDesiredSize:=vector2{X:=ActionIconSizeX,Y:=ActionIconSizeY}}
    PimentosResourceAmmount_TextBlock : text_block = text_block{DefaultTextColor:=PimentoTextColor, DefaultShadowOffset:=ShadowOffset, DefaultShadowOpacity := 1.0}
    var PimentosAmmounted : float = 0.0
    var PimentosResource_DisplayTime : int = 0
    var PimentosAmmountOnScreen : logic = false

    #Coal Resource
    CoalResource_Slider : texture_block = texture_block{DefaultDesiredSize:=vector2{X:=XSizeActionUI,Y:=OutOfScreenY},DefaultImage:=Images.EmptyTexture}
    CoalResource_Background: texture_block = texture_block{DefaultImage:=Images.BlackBackground}
    CoalResource_Icon : texture_block = texture_block{DefaultImage:=Images.Coal_Icon, DefaultDesiredSize:=vector2{X:=ActionIconSizeX,Y:=ActionIconSizeY}}
    CoalResourceAmmount_TextBlock : text_block = text_block{DefaultTextColor:=CoalTextColor, DefaultShadowOffset:=ShadowOffset, DefaultShadowOpacity := 1.0}
    var CoalAmmounted : float = 0.0
    var CoalResource_DisplayTime : int = 0
    var CoalAmmountOnScreen : logic = false

    #Mutant Resource
    MutantResource_Slider : texture_block = texture_block{DefaultDesiredSize:=vector2{X:=XSizeActionUI,Y:=OutOfScreenY},DefaultImage:=Images.EmptyTexture}
    MutantResource_Background: texture_block = texture_block{DefaultImage:=Images.BlackBackground}
    MutantResource_Icon : texture_block = texture_block{DefaultImage:=Images.Mutant_Material_Icon, DefaultDesiredSize:=vector2{X:=ActionIconSizeX,Y:=ActionIconSizeY}}
    MutantResourceAmmount_TextBlock : text_block = text_block{DefaultTextColor:=MutantMaterialTextColor, DefaultShadowOffset:=ShadowOffset, DefaultShadowOpacity := 1.0}
    var MutantAmmounted : float = 0.0
    var MutantResource_DisplayTime : int = 0
    var MutantAmmountOnScreen : logic = false

    #Mob Damage
    MobDamage_Slider : texture_block = texture_block{DefaultDesiredSize:=vector2{X:=XSizeActionUI,Y:=OutOfScreenY},DefaultImage:=Images.EmptyTexture}
    MobDamage_Background: texture_block = texture_block{DefaultImage:=Images.BlackBackground}
    MobDamage_Icon : texture_block = texture_block{DefaultImage:=Images.Coal_Icon, DefaultDesiredSize:=vector2{X:=ActionIconSizeX,Y:=ActionIconSizeY}}
    MobDamageAmmount_TextBlock : text_block = text_block{DefaultTextColor:=CoalTextColor, DefaultShadowOffset:=ShadowOffset, DefaultShadowOpacity := 1.0}
    var MobDamageAmmounted : float = 0.0
    var MobDamage_DisplayTime : int = 0
    var MobDamageAmmountOnScreen : logic = false


    var AllSlider_Texture : []texture_block = array{}
    var AllBackground_Texture : []texture_block = array{}
    var AllAction_Icon : []texture_block = array{}
    var AllAmmount_TextBlock : []text_block = array{}
    var AllUsedSliders : []logic = array{}

    var MaybeCanvas : ?canvas = false
    var UIIsDisplayed : logic = false

    var TimeElapsed : int = 0

    ##Close
    #CloseButton : button_loud = button_loud{}

    SetAgent(InAgent : agent) : void =
        Remove_CanvasWithoutMessage()
        set MaybeAgent = option{InAgent}
        set IsDungeonStarting = false
        set UIIsDisplayed = false

    InitUI() : void =
        set AllUsedSliders = array{Used_0,Used_1,Used_2,Used_3,Used_4,Used_5,Used_6,Used_7,Used_8,Used_9,Used_10,Used_11,Used_12,Used_13,Used_14,Used_15,Used_16}
        set AllSlider_Texture = array{Slider_Texture_0,Slider_Texture_1,Slider_Texture_2,Slider_Texture_3,Slider_Texture_4,Slider_Texture_5,Slider_Texture_6,Slider_Texture_7,Slider_Texture_8,Slider_Texture_9,Slider_Texture_10,Slider_Texture_11,Slider_Texture_12,Slider_Texture_13,Slider_Texture_14,Slider_Texture_15,Slider_Texture_16}
        set AllAction_Icon = array{Action_Icon_0,Action_Icon_1,Action_Icon_2,Action_Icon_3,Action_Icon_4,Action_Icon_5,Action_Icon_6,Action_Icon_7,Action_Icon_8,Action_Icon_9,Action_Icon_10,Action_Icon_11,Action_Icon_12,Action_Icon_13,Action_Icon_14,Action_Icon_15,Action_Icon_16}
        set AllAmmount_TextBlock = array{Ammount_TextBlock_0,Ammount_TextBlock_1,Ammount_TextBlock_2,Ammount_TextBlock_3,Ammount_TextBlock_4,Ammount_TextBlock_5,Ammount_TextBlock_6,Ammount_TextBlock_7,Ammount_TextBlock_8,Ammount_TextBlock_9,Ammount_TextBlock_10,Ammount_TextBlock_11,Ammount_TextBlock_12,Ammount_TextBlock_13,Ammount_TextBlock_14,Ammount_TextBlock_15,Ammount_TextBlock_16}

        Canvas := MakeCanvas()
        set MaybeCanvas = option{Canvas}

        <#
    Remove_Canvas(Message : widget_message) : void =
        set UIIsDisplayed = false
        ##Print("Trying to remove UI")
        if:
            Agent := MaybeAgent?
            Player:=player[Agent]
            PlayerUI:=GetPlayerUI[Player]
            Canvas := MaybeCanvas?
        then:
        #    #Print("Removing UI")
            PlayerUI.RemoveWidget(Canvas)#>

    Remove_CanvasWithoutMessage() : void =
        set UIIsDisplayed = false
        ##Print("Trying to remove UI")
        if:
            Agent := MaybeAgent?
            Player:=player[Agent]
            PlayerUI:=GetPlayerUI[Player]
            Canvas := MaybeCanvas?
        then:
        #    #Print("Removing UI")
            PlayerUI.RemoveWidget(Canvas)

    Display_Canvas() : void =
        set TimeElapsed = 0
        if(not UIIsDisplayed?):
            set UIIsDisplayed = true
            if:
                Agent := MaybeAgent?
                PlayerUI:=GetPlayerUI[player[Agent]]
                Canvas := MaybeCanvas?
            then:
                PlayerUI.AddWidget(Canvas,player_ui_slot{ZOrder:=0})
                RemoveDungeonInfo()
        else:
            PrintWarning("Can not display canvas on action info ui")
            
    DisplayGather(Ammounts : []float, Types : []int, AllLimitsReached : []logic, CriticalStrike : logic): void =
        spawn:
            DisplayMultiple(Ammounts,Types,AllLimitsReached,CriticalStrike)

    DisplayMultiple<private>(Ammounts: []float, Types: []int, AllLimitsReached : []logic, Criticial : logic)<suspends> : void =
        var SlidersToUse : []int = array{}
        NumSlidersToFind := AllLimitsReached.Length
        var SliderIndex : int = 0
        ##Print("Trying to show a resource")
        #Reserve sliders
        loop:
            if(IsUsed:=AllUsedSliders[SliderIndex], not IsUsed?, set AllUsedSliders[SliderIndex]=true):
                set SlidersToUse = SlidersToUse + array{SliderIndex}
                if(SlidersToUse.Length>=NumSlidersToFind):
                    break
            else if(SliderIndex>AllUsedSliders.Length-1):
                PrintWarning("WARNING: Failed trying to search a not used Slider")
                ClearSliders()
                break
            set SliderIndex = SliderIndex + 1

        for(Index:=0..AllLimitsReached.Length-1, IndexSlider := SlidersToUse[Index], Ammount := Ammounts[Index], Type := Types[Index], LimitReached := AllLimitsReached[Index]):
            Log("Displaying Type: {Type}. Amount: {Ammount}",LOG_LEVEL_DEBUG1)
            spawn: 
                StartUIAnimation(IndexSlider,Ammount,Type,LimitReached,Criticial)
            Sleep(0.15)
        

    AmmountTotal<private>(Ammount : float, Type : int, WasOverLimit : logic) : void =
        if(Ammount<>0.0):
            if(Type=WATER):
                if(Ammount>0.0):
                    WaterResourceAmmount_TextBlock.SetTextColor(WaterTextColor)
                else if(Ammount < 0.0):
                    WaterResourceAmmount_TextBlock.SetTextColor(Red)
                if(not WasOverLimit?):
                    set WaterAmmounted = WaterAmmounted + Ammount 
                    WaterResourceAmmount_TextBlock.SetText(BasicMessage("x {GetOptimizedMagnitudeLetter(WaterAmmounted)}"))
                else if(Agent := MaybeAgent?):
                    BakeryMachinesManager.GameManager.VFXManager.Full_AudioPlayer.Play(Agent)
                    WaterResourceAmmount_TextBlock.SetText(BasicMessage("FULL"))
            else if(Type=FLOUR):
                if(Ammount>0.0):
                    FlourResourceAmmount_TextBlock.SetTextColor(FlourTextColor)
                else if(Ammount < 0.0):
                    FlourResourceAmmount_TextBlock.SetTextColor(Red)
                if(not WasOverLimit?):
                    set FlourAmmounted = FlourAmmounted + Ammount 
                    FlourResourceAmmount_TextBlock.SetText(BasicMessage("x {GetOptimizedMagnitudeLetter(FlourAmmounted)}"))
                else if(Agent := MaybeAgent?):
                    BakeryMachinesManager.GameManager.VFXManager.Full_AudioPlayer.Play(Agent)
                    FlourResourceAmmount_TextBlock.SetText(BasicMessage("FULL"))
            else if(Type=CHEESE):
                if(Ammount>0.0):
                    CheeseResourceAmmount_TextBlock.SetTextColor(CheeseTextColor)
                else if(Ammount < 0.0):
                    CheeseResourceAmmount_TextBlock.SetTextColor(Red)
                if(not WasOverLimit?):
                    set CheeseAmmounted = CheeseAmmounted + Ammount 
                    CheeseResourceAmmount_TextBlock.SetText(BasicMessage("x {GetOptimizedMagnitudeLetter(CheeseAmmounted)}"))
                else if(Agent := MaybeAgent?):
                    BakeryMachinesManager.GameManager.VFXManager.Full_AudioPlayer.Play(Agent)
                    CheeseResourceAmmount_TextBlock.SetText(BasicMessage("FULL"))
            else if(Type=PIMENTOS):
                if(Ammount>0.0):
                    PimentosResourceAmmount_TextBlock.SetTextColor(PimentoTextColor)
                else if(Ammount < 0.0):
                    PimentosResourceAmmount_TextBlock.SetTextColor(Red)
                if(not WasOverLimit?):
                    set PimentosAmmounted = PimentosAmmounted + Ammount 
                    PimentosResourceAmmount_TextBlock.SetText(BasicMessage("x {GetOptimizedMagnitudeLetter(PimentosAmmounted)}"))
                else if(Agent := MaybeAgent?):
                    BakeryMachinesManager.GameManager.VFXManager.Full_AudioPlayer.Play(Agent)
                    PimentosResourceAmmount_TextBlock.SetText(BasicMessage("FULL"))
            else if(Type=COAL):
                if(Ammount>0.0):
                    CoalResourceAmmount_TextBlock.SetTextColor(CoalTextColor)
                else if(Ammount < 0.0):
                    CoalResourceAmmount_TextBlock.SetTextColor(Red)
                if(not WasOverLimit?):
                    set CoalAmmounted = CoalAmmounted + Ammount 
                    CoalResourceAmmount_TextBlock.SetText(BasicMessage("x {GetOptimizedMagnitudeLetter(CoalAmmounted)}"))
                else:
                    CoalResourceAmmount_TextBlock.SetText(BasicMessage("FULL"))
            else if(Type=MUTANTMATERIAL):
                if(Ammount>0.0):
                    MutantResourceAmmount_TextBlock.SetTextColor(MutantMaterialTextColor)
                else if(Ammount < 0.0):
                    MutantResourceAmmount_TextBlock.SetTextColor(Red)
                if(not WasOverLimit?):
                    set MutantAmmounted = MutantAmmounted + Ammount 
                    MutantResourceAmmount_TextBlock.SetText(BasicMessage("x {GetOptimizedMagnitudeLetter(MutantAmmounted)}"))
                else:
                    MutantResourceAmmount_TextBlock.SetText(BasicMessage("FULL"))
            spawn:
                RunUIRemoveByTime(Type)
            
    RunUIRemoveByTime<private>(Type : int)<suspends> : void =
        if(Type=WATER):
            set WaterResource_DisplayTime = 0
            if(not WaterAmmountOnScreen?):
                loop:   
                    WaterResource_Slider.SetDesiredSize(vector2{X:=XSizeActionUI,Y:=0.0})
                    set WaterAmmountOnScreen = true
                    if(WaterResource_DisplayTime >= MaxActionDisplayTime):
                        WaterResource_Slider.SetDesiredSize(vector2{X:=XSizeActionUI,Y:=OutOfScreenY})
                        set WaterAmmountOnScreen = false
                        set WaterAmmounted = 0.0
                        break
                    Sleep(1.0)
                    set WaterResource_DisplayTime = WaterResource_DisplayTime + 1
        else if(Type=FLOUR):
            set FlourResource_DisplayTime = 0
            if(not FlourAmmountOnScreen?):
                loop:   
                    FlourResource_Slider.SetDesiredSize(vector2{X:=XSizeActionUI,Y:=0.0})
                    set FlourAmmountOnScreen = true
                    if(FlourResource_DisplayTime >= MaxActionDisplayTime):
                        FlourResource_Slider.SetDesiredSize(vector2{X:=XSizeActionUI,Y:=OutOfScreenY})
                        set FlourAmmountOnScreen = false
                        set FlourAmmounted = 0.0
                        break
                    Sleep(1.0)
                    set FlourResource_DisplayTime = FlourResource_DisplayTime + 1
        else if(Type=CHEESE):
            set CheeseResource_DisplayTime = 0
            if(not CheeseAmmountOnScreen?):
                loop:   
                    CheeseResource_Slider.SetDesiredSize(vector2{X:=XSizeActionUI,Y:=0.0})
                    set CheeseAmmountOnScreen = true
                    if(CheeseResource_DisplayTime >= MaxActionDisplayTime):
                        CheeseResource_Slider.SetDesiredSize(vector2{X:=XSizeActionUI,Y:=OutOfScreenY})
                        set CheeseAmmountOnScreen = false
                        set CheeseAmmounted = 0.0
                        break
                    Sleep(1.0)
                    set CheeseResource_DisplayTime = CheeseResource_DisplayTime + 1
        else if(Type=PIMENTOS):
            set PimentosResource_DisplayTime = 0
            if(not PimentosAmmountOnScreen?):
                loop:   
                    PimentosResource_Slider.SetDesiredSize(vector2{X:=XSizeActionUI,Y:=0.0})
                    set PimentosAmmountOnScreen = true
                    if(PimentosResource_DisplayTime >= MaxActionDisplayTime):
                        PimentosResource_Slider.SetDesiredSize(vector2{X:=XSizeActionUI,Y:=OutOfScreenY})
                        set PimentosAmmountOnScreen = false
                        set PimentosAmmounted = 0.0
                        break
                    Sleep(1.0)
                    set PimentosResource_DisplayTime = PimentosResource_DisplayTime + 1
        else if(Type=COAL):
            set CoalResource_DisplayTime = 0
            if(not CoalAmmountOnScreen?):
                loop:   
                    CoalResource_Slider.SetDesiredSize(vector2{X:=XSizeActionUI,Y:=0.0})
                    set CoalAmmountOnScreen = true
                    if(CoalResource_DisplayTime >= MaxActionDisplayTime):
                        CoalResource_Slider.SetDesiredSize(vector2{X:=XSizeActionUI,Y:=OutOfScreenY})
                        set CoalAmmountOnScreen = false
                        set CoalAmmounted = 0.0
                        break
                    Sleep(1.0)
                    set CoalResource_DisplayTime = CoalResource_DisplayTime + 1
        else if(Type=MUTANTMATERIAL):
            set MutantResource_DisplayTime = 0
            if(not MutantAmmountOnScreen?):
                loop:   
                    MutantResource_Slider.SetDesiredSize(vector2{X:=XSizeActionUI,Y:=0.0})
                    set MutantAmmountOnScreen = true
                    if(MutantResource_DisplayTime >= MaxActionDisplayTime):
                        MutantResource_Slider.SetDesiredSize(vector2{X:=XSizeActionUI,Y:=OutOfScreenY})
                        set MutantAmmountOnScreen = false
                        set MutantAmmounted = 0.0
                        break
                    Sleep(1.0)
                    set MutantResource_DisplayTime = MutantResource_DisplayTime + 1
        else if(Type=SWORD_TYPE or Type=BOW_TYPE):
            set MobDamage_DisplayTime = 0
            if(not MobDamageAmmountOnScreen?):
                loop:   
                    MobDamage_Slider.SetDesiredSize(vector2{X:=XSizeActionUI,Y:=0.0})
                    set MobDamageAmmountOnScreen = true
                    if(MobDamage_DisplayTime >= MaxActionDisplayTime):
                        MobDamage_Slider.SetDesiredSize(vector2{X:=XSizeActionUI,Y:=OutOfScreenY})
                        set MobDamageAmmountOnScreen = false
                        set MobDamageAmmounted = 0.0
                        break
                    Sleep(1.0)
                    set MobDamage_DisplayTime = MobDamage_DisplayTime + 1
        #CheckSafeClearSliders()

    ClearSliders() : void =
        for(Index:=0..AllUsedSliders.Length-1, set AllUsedSliders[Index] = false):


    StartUIAnimation<private>(IndexSlider : int, Ammount : float, Type : int, LimitReached : logic, CriticalStrike : logic)<suspends> : void =
        if(Ammount<>0.0):
            AmmountTotal(Ammount,Type,LimitReached)
            FramesPerAction_1 : int = 5
            FramesPerAction_2 : int = 30

            #Phase1
            AddX : float = ActionIconSizeX/(FramesPerAction_1*1.0)
            AddY : float = ActionIconSizeY/(FramesPerAction_1*1.0)
            RemoveX : float = ActionIconSizeX/((FramesPerAction_2*1.0/2.0)*1.0)
            RemoveY : float = ActionIconSizeY/((FramesPerAction_2*1.0/2.0)*1.0)
            SmallAmmount:=Ammount/(FramesPerAction_1*1.0)
            var SizeVector : vector2 = vector2{X:=AddX,Y:=AddY}
            var DeSizeVector : vector2 = vector2{X:=-RemoveX,Y:=-RemoveY}
            var TotalVector : vector2 = vector2{X:=0.0,Y:=0.0}
            var TotalVector2 : vector2 = vector2{X:=ActionIconSizeX,Y:=ActionIconSizeY}

            var Background_Texture : texture_block = texture_block{DefaultImage:=Images.BlackBackground}
            #Log("StartUIAnimation, index: {Index}",LOG_LEVEL_UIs)

            if:
                Slider_Texture:=AllSlider_Texture[IndexSlider]
                Action_Icon:=AllAction_Icon[IndexSlider]
                Ammount_TextBlock:=AllAmmount_TextBlock[IndexSlider]
                OwnerAgent:=BakeryMachinesManager.MaybeAgentOwner?
            then:
                if(CriticalStrike?):
                    Ammount_TextBlock.SetShadowOffset(option{vector2{X:=3.0,Y:=3.0}})
                    #Ammount_TextBlock.SetShadowColor(Red)
                else:
                    Ammount_TextBlock.SetShadowOffset(ShadowOffset)
                    #Ammount_TextBlock.SetShadowColor(Black)

                Action_Icon.SetImage(Images.Bakery_Knife_Icon)
                Action_Icon.SetDesiredSize(TotalVector)

                Ammount_TextBlock.SetTextOpacity(1.0)
                Ammount_TextBlock.SetShadowOpacity(1.0)
                Ammount_TextBlock.SetText(BasicMessage("x {0.0}"))

                Slider_Texture.SetDesiredSize(vector2{X:=XSizeActionUI,Y:=20.0})

                Log("Displaying type = {Type}",LOG_LEVEL_1DETAILING)

                if(Type = WATER):
                    Log("Displaying type = WATER",LOG_LEVEL_1DETAILING)
                    Ammount_TextBlock.SetTextColor(WaterTextColor)
                    Action_Icon.SetImage(Images.Water_Icon)
                else if(Type = FLOUR):
                    Ammount_TextBlock.SetTextColor(FlourTextColor)
                    Action_Icon.SetImage(Images.Flour_Icon)
                else if(Type = CHEESE):
                    Ammount_TextBlock.SetTextColor(CheeseTextColor)
                    Action_Icon.SetImage(Images.Cheese_Icon)
                else if(Type = PIMENTOS):
                    Ammount_TextBlock.SetTextColor(PimentoTextColor)
                    Action_Icon.SetImage(Images.Pimento_Icon)
                else if(Type = MUTANTMATERIAL):
                    Ammount_TextBlock.SetTextColor(MutantMaterialTextColor)
                    Action_Icon.SetImage(Images.Mutant_Material_Icon)
                    if(Ammount>0.0):
                        BakeryMachinesManager.GameManager.VFXManager.Audio_MutantMaterialGranted.Play(OwnerAgent)
                else if(Type = COAL):
                    Ammount_TextBlock.SetTextColor(CoalTextColor)
                    Action_Icon.SetImage(Images.Coal_Icon)
                    if(Ammount>0.0):
                        BakeryMachinesManager.GameManager.VFXManager.Audio_CoalGranted.Play(OwnerAgent)
                else if(Type = BOW_TYPE):
                    Log("Displaying type = BOW_TYPE",LOG_LEVEL_1DETAILING)
                    Ammount_TextBlock.SetTextColor(MobDamageTextColor)
                    Action_Icon.SetImage(Images.Bow_Icon)
                else if(Type = SWORD_TYPE):
                    Log("Displaying type = SWORD_TYPE",LOG_LEVEL_1DETAILING)
                    Ammount_TextBlock.SetTextColor(MobDamageTextColor)
                    Action_Icon.SetImage(Images.Sword_Icon)
                else if(Type = SAPPHIRES):
                    Ammount_TextBlock.SetTextColor(SapphiresTextColor)
                    Action_Icon.SetImage(Images.Sapphire_Icon)
                    if(Ammount>0.0):
                        BakeryMachinesManager.GameManager.VFXManager.Audio_SapphiresGranted.Play(OwnerAgent)
                else if(Type = AGEDSAPPHIRES):
                    Ammount_TextBlock.SetTextColor(AgedSapphiresTextColor)
                    Action_Icon.SetImage(Images.AgedSaphire_Icon)
                    if(Ammount>0.0):
                        BakeryMachinesManager.GameManager.VFXManager.Audio_AgedSapphiresGranted.Play(OwnerAgent)
                else if(Type = HEXATOKEN):
                    Ammount_TextBlock.SetTextColor(HexaTokenTextColor)
                    Action_Icon.SetImage(Images.HexTrophy)
                else if(Type = DUNGEONTOKEN):
                    Ammount_TextBlock.SetTextColor(DungeonTokenTextColor)
                    Action_Icon.SetImage(Images.Dungeon_Icon)
                else if(Type = FAME):
                    if(Ammount>0.0):
                        BakeryMachinesManager.GameManager.VFXManager.Audio_FameGranted.Play(OwnerAgent)
                    else if(Ammount<0.0):
                        BakeryMachinesManager.GameManager.VFXManager.Audio_LessFameGranted.Play(OwnerAgent)
                    Ammount_TextBlock.SetTextColor(FameTextColor)
                    Action_Icon.SetImage(Images.Fame_Icon)
                else if(Type = ARROW):
                    Ammount_TextBlock.SetTextColor(White)
                    Action_Icon.SetImage(Images.Arrow_Icon)
                if(Ammount<0.0):
                    Ammount_TextBlock.SetTextColor(Red)
                    Action_Icon.SetTint(Red)
                else:
                    Action_Icon.SetTint(White)
                    
                #First phase
                for(Iteration:=1..FramesPerAction_1):
                    var AmmountPiece : float = Iteration * SmallAmmount
                    set TotalVector = TotalVector + SizeVector
                    Action_Icon.SetDesiredSize(TotalVector)
                    if(LimitReached?):
                        Ammount_TextBlock.SetText(BasicMessage("FULL"))
                    else:
                        Ammount_TextBlock.SetText(BasicMessage("x {GetOptimizedMagnitudeLetter(AmmountPiece)}"))
                    Sleep(0.04)

                var SliderEffect : float = 500.0/(FramesPerAction_2*1.0)
                var OpacityEffect : float = 1.0 /(FramesPerAction_2*1.0) 
                #SecondPhase
                for(Iteration:=1..FramesPerAction_2):
                    var ReverseFloat : float = 1.0 - (OpacityEffect*Iteration)
                    if (Opacity := type{X:float where 0.0 <= X, X <= 1.0}[ReverseFloat]):
                        Ammount_TextBlock.SetTextOpacity(Opacity)
                        Ammount_TextBlock.SetShadowOpacity(Opacity)
                        Slider_Texture.SetDesiredSize(vector2{X:=XSizeActionUI,Y:=SliderEffect*Iteration})
                    if(Iteration*1.0>(FramesPerAction_2*1.0/2.0)):
                        set TotalVector2 = TotalVector2 + DeSizeVector
                        Action_Icon.SetDesiredSize(TotalVector2)
                    Sleep(0.04)
                Slider_Texture.SetDesiredSize(vector2{X:=XSizeActionUI,Y:=OutOfScreenY})
            else:
                Log("FATAL ERROR, index: {IndexSlider} not found!",LOG_LEVEL_UIs)
        if(set AllUsedSliders[IndexSlider]=false):

    Set_SimonInfo(InfoString : string) : void=
        ##Print("Duneon info set")
        #set IsDungeonStarting = true
        Simon_Slider_Texture.SetDesiredSize(vector2{X:=0.0,Y:=0.0})
        Simon_TextBlock.SetText(BasicMessage(InfoString))
        Simon_TextBlock.SetTextColor(Gold)

    RemoveSimonInfo() : void =
        ##Print("Dungeon info removed")
        Simon_Slider_Texture.SetDesiredSize(vector2{X:=1.0,Y:=2200.0})

    SetDungeon_TimeInfo(InfoString : string) : void=
        ##Print("Duneon info set")
        set IsDungeonStarting = true
        Dungeon_Slider_Texture.SetDesiredSize(vector2{X:=0.0,Y:=0.0})
        Dungeon_TextBlock.SetText(BasicMessage(InfoString))
        Dungeon_TextBlock.SetTextColor(White)

    SetDungeon_TimeEndingInfo(InfoString : string, Time : int) : void=
        ##Print("Duneon info set")
        Dungeon_Slider_Texture.SetDesiredSize(vector2{X:=0.0,Y:=0.0})
        if(IsOdd[Time]):
            Dungeon_TextBlock.SetTextColor(Red)
        else:
            Dungeon_TextBlock.SetTextColor(Yellow)
        Dungeon_TextBlock.SetText(BasicMessage(InfoString+ "{Time}"))

    SetDungeon_ChestInfo(InfoString : string) : void=
        ##Print("Duneon info set")
        set IsDungeonStarting = false
        Dungeon_Slider_Texture.SetDesiredSize(vector2{X:=0.0,Y:=0.0})
        DungeonRemainingChests_TextBlock.SetText(BasicMessage(InfoString))

    RemoveDungeonInfo() : void =
        ##Print("Dungeon info removed")
        set IsDungeonStarting = false
        Dungeon_Slider_Texture.SetDesiredSize(vector2{X:=1.0,Y:=2200.0})

    MakeCanvas(): canvas=
        Canvas := canvas:
            Slots:=array:
                #Simon event info
                canvas_slot:
                    Anchors:=anchors{Minimum:=vector2{X:=0.5,Y:=0.0},Maximum:=vector2{X:=0.5,Y:=0.0}}
                    ZOrder:=0
                    SizeToContent:=true
                    Offsets:=margin{Left:=0.0,Top:=160.0,Right:=0.0,Bottom:=0.0} #PositionX, PositionY, SizeX, SizeY
                    Alignment:=vector2{X:=0.5,Y:=0.0}
                    Widget:=overlay:
                        Slots:=array:
                            overlay_slot:
                                HorizontalAlignment:=horizontal_alignment.Center
                                VerticalAlignment:=vertical_alignment.Center
                                Widget:=overlay:
                                    Slots:=array:
                                        overlay_slot:
                                            Widget:=Simon_Slider_Texture
                            overlay_slot:
                                HorizontalAlignment:=horizontal_alignment.Center
                                VerticalAlignment:=vertical_alignment.Bottom
                                Widget:=overlay:
                                    Slots:=array:
                                        overlay_slot:
                                            HorizontalAlignment:=horizontal_alignment.Fill
                                            VerticalAlignment:=vertical_alignment.Fill
                                            Widget:=texture_block{DefaultImage:=Images.Img_LineBackground, DefaultTint:=Gold, DefaultDesiredSize:=vector2{X:=1.0,Y:=1.0}}
                                        overlay_slot:
                                            HorizontalAlignment:=horizontal_alignment.Fill
                                            VerticalAlignment:=vertical_alignment.Fill
                                            Padding:=margin{Left:=2.5,Top:=2.5,Right:=2.5,Bottom:=2.5}
                                            Widget:=texture_block{DefaultImage:=Images.Img_LineBackground, DefaultTint:=Gold, DefaultDesiredSize:=vector2{X:=1.0,Y:=1.0}}
                                        overlay_slot:
                                            HorizontalAlignment:=horizontal_alignment.Fill
                                            VerticalAlignment:=vertical_alignment.Fill
                                            Widget:=texture_block{DefaultDesiredSize:=vector2{X:=50.0,Y:=50.0},DefaultTint:=Gold,DefaultImage:=Images.Img_TintedBackgroundHighOpacity}
                                        overlay_slot: #Screw 0
                                            Padding:=margin{Left:=3.5,Top:=3.5,Right:=3.5,Bottom:=3.5}
                                            HorizontalAlignment:=horizontal_alignment.Left
                                            VerticalAlignment:=vertical_alignment.Top
                                            Widget:=texture_block{DefaultImage:=Images.Dialogue_Bolt, DefaultDesiredSize:=vector2{X:=5.0,Y:=5.0}}
                                        overlay_slot: #Screw 1
                                            Padding:=margin{Left:=3.5,Top:=3.5,Right:=3.5,Bottom:=3.5}
                                            HorizontalAlignment:=horizontal_alignment.Right
                                            VerticalAlignment:=vertical_alignment.Top
                                            Widget:=texture_block{DefaultImage:=Images.Dialogue_Bolt, DefaultDesiredSize:=vector2{X:=5.0,Y:=5.0}}
                                        overlay_slot: #Screw 2
                                            Padding:=margin{Left:=3.5,Top:=3.5,Right:=3.5,Bottom:=3.5}
                                            HorizontalAlignment:=horizontal_alignment.Left
                                            VerticalAlignment:=vertical_alignment.Bottom
                                            Widget:=texture_block{DefaultImage:=Images.Dialogue_Bolt, DefaultDesiredSize:=vector2{X:=5.0,Y:=5.0}}
                                        overlay_slot: #Screw 3
                                            Padding:=margin{Left:=3.5,Top:=3.5,Right:=3.5,Bottom:=3.5}
                                            HorizontalAlignment:=horizontal_alignment.Right
                                            VerticalAlignment:=vertical_alignment.Bottom
                                            Widget:=texture_block{DefaultImage:=Images.Dialogue_Bolt, DefaultDesiredSize:=vector2{X:=5.0,Y:=5.0}}
                                        overlay_slot:
                                            HorizontalAlignment:=horizontal_alignment.Center
                                            VerticalAlignment:=vertical_alignment.Center
                                            Padding:=margin{Left:=RowLeftPadding,Top:=RowTopPadding,Right:=RowRightPadding,Bottom:=RowBottomPadding}
                                            Widget:=stack_box:
                                                Orientation:=orientation.Horizontal
                                                    Slots:=array:
                                                        stack_box_slot:
                                                            Padding:=margin{Left:=5.0,Top:=2.5,Right:=5.0,Bottom:=2.5}
                                                            HorizontalAlignment:=horizontal_alignment.Center
                                                            VerticalAlignment:=vertical_alignment.Center
                                                            Widget:=Simon_Icon_Texture
                                                        stack_box_slot:
                                                            Padding:=margin{Left:=5.0,Top:=2.5,Right:=5.0,Bottom:=2.5}
                                                            HorizontalAlignment:=horizontal_alignment.Center
                                                            VerticalAlignment:=vertical_alignment.Center
                                                            Widget:=Simon_TextBlock
                #Dungeon info
                canvas_slot:
                    Anchors:=anchors{Minimum:=vector2{X:=0.5,Y:=0.0},Maximum:=vector2{X:=0.5,Y:=0.0}}
                    ZOrder:=0
                    SizeToContent:=true
                    Offsets:=margin{Left:=0.0,Top:=50.0,Right:=0.0,Bottom:=0.0} #PositionX, PositionY, SizeX, SizeY
                    Alignment:=vector2{X:=0.5,Y:=0.0}
                    Widget:=overlay:
                        Slots:=array:
                            overlay_slot:
                                HorizontalAlignment:=horizontal_alignment.Center
                                VerticalAlignment:=vertical_alignment.Center
                                Widget:=overlay:
                                    Slots:=array:
                                        overlay_slot:
                                            Widget:=Dungeon_Slider_Texture
                            overlay_slot:
                                HorizontalAlignment:=horizontal_alignment.Center
                                VerticalAlignment:=vertical_alignment.Bottom
                                Widget:=overlay:
                                    Slots:=array:
                                        overlay_slot:
                                            HorizontalAlignment:=horizontal_alignment.Fill
                                            VerticalAlignment:=vertical_alignment.Fill
                                            Widget:=texture_block{DefaultImage:=Images.Img_LineBackground, DefaultDesiredSize:=vector2{X:=1.0,Y:=1.0}}
                                        overlay_slot:
                                            HorizontalAlignment:=horizontal_alignment.Fill
                                            VerticalAlignment:=vertical_alignment.Fill
                                            Padding:=margin{Left:=2.5,Top:=2.5,Right:=2.5,Bottom:=2.5}
                                            Widget:=texture_block{DefaultImage:=Images.Img_LineBackground, DefaultDesiredSize:=vector2{X:=1.0,Y:=1.0}}
                                        overlay_slot:
                                            HorizontalAlignment:=horizontal_alignment.Fill
                                            VerticalAlignment:=vertical_alignment.Fill
                                            Widget:=texture_block{DefaultDesiredSize:=vector2{X:=200.0,Y:=100.0},DefaultTint:=Black,DefaultImage:=Images.Img_TintedBackgroundHighOpacity}
                                        overlay_slot: #Screw 0
                                            Padding:=margin{Left:=3.5,Top:=3.5,Right:=3.5,Bottom:=3.5}
                                            HorizontalAlignment:=horizontal_alignment.Left
                                            VerticalAlignment:=vertical_alignment.Top
                                            Widget:=texture_block{DefaultImage:=Images.Dialogue_Bolt, DefaultDesiredSize:=vector2{X:=5.0,Y:=5.0}}
                                        overlay_slot: #Screw 1
                                            Padding:=margin{Left:=3.5,Top:=3.5,Right:=3.5,Bottom:=3.5}
                                            HorizontalAlignment:=horizontal_alignment.Right
                                            VerticalAlignment:=vertical_alignment.Top
                                            Widget:=texture_block{DefaultImage:=Images.Dialogue_Bolt, DefaultDesiredSize:=vector2{X:=5.0,Y:=5.0}}
                                        overlay_slot: #Screw 2
                                            Padding:=margin{Left:=3.5,Top:=3.5,Right:=3.5,Bottom:=3.5}
                                            HorizontalAlignment:=horizontal_alignment.Left
                                            VerticalAlignment:=vertical_alignment.Bottom
                                            Widget:=texture_block{DefaultImage:=Images.Dialogue_Bolt, DefaultDesiredSize:=vector2{X:=5.0,Y:=5.0}}
                                        overlay_slot: #Screw 3
                                            Padding:=margin{Left:=3.5,Top:=3.5,Right:=3.5,Bottom:=3.5}
                                            HorizontalAlignment:=horizontal_alignment.Right
                                            VerticalAlignment:=vertical_alignment.Bottom
                                            Widget:=texture_block{DefaultImage:=Images.Dialogue_Bolt, DefaultDesiredSize:=vector2{X:=5.0,Y:=5.0}}
                                        overlay_slot:
                                            HorizontalAlignment:=horizontal_alignment.Center
                                            VerticalAlignment:=vertical_alignment.Center
                                            Padding:=margin{Left:=RowLeftPadding,Top:=RowTopPadding,Right:=RowRightPadding,Bottom:=RowBottomPadding}
                                            Widget:=stack_box:
                                                Orientation:=orientation.Vertical
                                                    Slots:=array:
                                                        stack_box_slot:
                                                            Padding:=margin{Left:=5.0,Top:=2.5,Right:=5.0,Bottom:=2.5}
                                                            HorizontalAlignment:=horizontal_alignment.Center
                                                            VerticalAlignment:=vertical_alignment.Center
                                                            Widget:=Dungeon_TextBlock
                                                        stack_box_slot:
                                                            Padding:=margin{Left:=5.0,Top:=2.5,Right:=5.0,Bottom:=2.5}
                                                            HorizontalAlignment:=horizontal_alignment.Center
                                                            VerticalAlignment:=vertical_alignment.Center
                                                            Widget:=DungeonRemainingChests_TextBlock
                #Moving sliders
                #Moving slider 0
                canvas_slot:
                    Anchors:=anchors{Minimum:=vector2{X:=0.5,Y:=0.5},Maximum:=vector2{X:=0.5,Y:=0.5}}
                    ZOrder:=0
                    SizeToContent:=true
                    Offsets:=margin{Left:=XPositionActionUI_Action,Top:=YPositionActionUI_Action,Right:=0.0,Bottom:=0.0} #PositionX, PositionY, SizeX, SizeY
                    Alignment:=vector2{X:=0.5,Y:=0.5}
                    Widget:=overlay:
                        Slots:=array:
                            overlay_slot:
                                HorizontalAlignment:=horizontal_alignment.Center
                                VerticalAlignment:=vertical_alignment.Center
                                Widget:=overlay:
                                    Slots:=array:
                                        overlay_slot:
                                            Widget:=Slider_Texture_0
                            overlay_slot:
                                HorizontalAlignment:=horizontal_alignment.Center
                                VerticalAlignment:=vertical_alignment.Top
                                Widget:=overlay:
                                    Slots:=array:
                                        overlay_slot:
                                            HorizontalAlignment:=horizontal_alignment.Center
                                            VerticalAlignment:=vertical_alignment.Center
                                            Widget:=texture_block{DefaultDesiredSize:=vector2{X:=XSizeActionUI,Y:=YSizeActionUI},DefaultImage:=Images.EmptyTexture}
                                        overlay_slot:
                                            HorizontalAlignment:=horizontal_alignment.Left
                                            VerticalAlignment:=vertical_alignment.Center
                                            Padding:=margin{Left:=RowLeftPadding,Top:=RowTopPadding,Right:=RowRightPadding,Bottom:=RowBottomPadding}
                                            Widget:=stack_box:
                                                Orientation:=orientation.Horizontal
                                                    Slots:=array:
                                                        stack_box_slot:
                                                            Padding:=margin{Left:=InterColumnLeftPadding,Top:=InterColumnTopPadding,Right:=InterColumnRightPadding,Bottom:=InterColumnBottomPadding}
                                                            HorizontalAlignment:=horizontal_alignment.Left
                                                            VerticalAlignment:=vertical_alignment.Center
                                                            Widget:=Action_Icon_0
                                                        stack_box_slot:
                                                            Padding:=margin{Left:=InterColumnLeftPadding,Top:=InterColumnTopPadding,Right:=InterColumnRightPadding,Bottom:=InterColumnBottomPadding}
                                                            HorizontalAlignment:=horizontal_alignment.Center
                                                            VerticalAlignment:=vertical_alignment.Center
                                                            Widget:=Ammount_TextBlock_0  
                #Moving slider 1 
                canvas_slot:
                    Anchors:=anchors{Minimum:=vector2{X:=0.5,Y:=0.5},Maximum:=vector2{X:=0.5,Y:=0.5}}
                    ZOrder:=0
                    SizeToContent:=true
                    Offsets:=margin{Left:=XPositionActionUI_Action,Top:=YPositionActionUI_Action,Right:=0.0,Bottom:=0.0} #PositionX, PositionY, SizeX, SizeY
                    Alignment:=vector2{X:=0.5,Y:=0.5}
                    Widget:=overlay:
                        Slots:=array:
                            overlay_slot:
                                HorizontalAlignment:=horizontal_alignment.Center
                                VerticalAlignment:=vertical_alignment.Center
                                Widget:=overlay:
                                    Slots:=array:
                                        overlay_slot:
                                            Widget:=Slider_Texture_1
                            overlay_slot:
                                HorizontalAlignment:=horizontal_alignment.Center
                                VerticalAlignment:=vertical_alignment.Top
                                Widget:=overlay:
                                    Slots:=array:
                                        overlay_slot:
                                            HorizontalAlignment:=horizontal_alignment.Center
                                            VerticalAlignment:=vertical_alignment.Center
                                            Widget:=texture_block{DefaultDesiredSize:=vector2{X:=XSizeActionUI,Y:=YSizeActionUI},DefaultImage:=Images.EmptyTexture}
                                        <#overlay_slot:
                                            HorizontalAlignment:=horizontal_alignment.Fill
                                            VerticalAlignment:=vertical_alignment.Fill
                                            Widget:=overlay:
                                                Slots:=array:
                                                    overlay_slot:
                                                        HorizontalAlignment:=horizontal_alignment.Fill
                                                        VerticalAlignment:=vertical_alignment.Fill
                                                        Widget:=texture_block{DefaultImage:=Images.Img_LineBackground}#>
                                        overlay_slot:
                                            HorizontalAlignment:=horizontal_alignment.Left
                                            VerticalAlignment:=vertical_alignment.Center
                                            Padding:=margin{Left:=RowLeftPadding,Top:=RowTopPadding,Right:=RowRightPadding,Bottom:=RowBottomPadding}
                                            Widget:=stack_box:
                                                Orientation:=orientation.Horizontal
                                                    Slots:=array:
                                                        stack_box_slot:
                                                            Padding:=margin{Left:=InterColumnLeftPadding,Top:=InterColumnTopPadding,Right:=InterColumnRightPadding,Bottom:=InterColumnBottomPadding}
                                                            HorizontalAlignment:=horizontal_alignment.Left
                                                            VerticalAlignment:=vertical_alignment.Center
                                                            Widget:=Action_Icon_1
                                                        stack_box_slot:
                                                            Padding:=margin{Left:=InterColumnLeftPadding,Top:=InterColumnTopPadding,Right:=InterColumnRightPadding,Bottom:=InterColumnBottomPadding}
                                                            HorizontalAlignment:=horizontal_alignment.Center
                                                            VerticalAlignment:=vertical_alignment.Center
                                                            Widget:=Ammount_TextBlock_1  
                #Moving slider 2 
                canvas_slot:
                    Anchors:=anchors{Minimum:=vector2{X:=0.5,Y:=0.5},Maximum:=vector2{X:=0.5,Y:=0.5}}
                    ZOrder:=0
                    SizeToContent:=true
                    Offsets:=margin{Left:=XPositionActionUI_Action,Top:=YPositionActionUI_Action,Right:=0.0,Bottom:=0.0} #PositionX, PositionY, SizeX, SizeY
                    Alignment:=vector2{X:=0.5,Y:=0.5}
                    Widget:=overlay:
                        Slots:=array:
                            overlay_slot:
                                HorizontalAlignment:=horizontal_alignment.Center
                                VerticalAlignment:=vertical_alignment.Center
                                Widget:=overlay:
                                    Slots:=array:
                                        overlay_slot:
                                            Widget:=Slider_Texture_2
                            overlay_slot:
                                HorizontalAlignment:=horizontal_alignment.Center
                                VerticalAlignment:=vertical_alignment.Top
                                Widget:=overlay:
                                    Slots:=array:
                                        overlay_slot:
                                            HorizontalAlignment:=horizontal_alignment.Center
                                            VerticalAlignment:=vertical_alignment.Center
                                            Widget:=texture_block{DefaultDesiredSize:=vector2{X:=XSizeActionUI,Y:=YSizeActionUI},DefaultImage:=Images.EmptyTexture}
                                        <#overlay_slot:
                                            HorizontalAlignment:=horizontal_alignment.Fill
                                            VerticalAlignment:=vertical_alignment.Fill
                                            Widget:=overlay:
                                                Slots:=array:
                                                    overlay_slot:
                                                        HorizontalAlignment:=horizontal_alignment.Fill
                                                        VerticalAlignment:=vertical_alignment.Fill
                                                        Widget:=texture_block{DefaultImage:=Images.Img_LineBackground}#>
                                        overlay_slot:
                                            HorizontalAlignment:=horizontal_alignment.Left
                                            VerticalAlignment:=vertical_alignment.Center
                                            Padding:=margin{Left:=RowLeftPadding,Top:=RowTopPadding,Right:=RowRightPadding,Bottom:=RowBottomPadding}
                                            Widget:=stack_box:
                                                Orientation:=orientation.Horizontal
                                                    Slots:=array:
                                                        stack_box_slot:
                                                            Padding:=margin{Left:=InterColumnLeftPadding,Top:=InterColumnTopPadding,Right:=InterColumnRightPadding,Bottom:=InterColumnBottomPadding}
                                                            HorizontalAlignment:=horizontal_alignment.Left
                                                            VerticalAlignment:=vertical_alignment.Center
                                                            Widget:=Action_Icon_2
                                                        stack_box_slot:
                                                            Padding:=margin{Left:=InterColumnLeftPadding,Top:=InterColumnTopPadding,Right:=InterColumnRightPadding,Bottom:=InterColumnBottomPadding}
                                                            HorizontalAlignment:=horizontal_alignment.Center
                                                            VerticalAlignment:=vertical_alignment.Center
                                                            Widget:=Ammount_TextBlock_2    

                #Moving slider 3 
                canvas_slot:
                    Anchors:=anchors{Minimum:=vector2{X:=0.5,Y:=0.5},Maximum:=vector2{X:=0.5,Y:=0.5}}
                    ZOrder:=0
                    SizeToContent:=true
                    Offsets:=margin{Left:=XPositionActionUI_Action,Top:=YPositionActionUI_Action,Right:=0.0,Bottom:=0.0} #PositionX, PositionY, SizeX, SizeY
                    Alignment:=vector2{X:=0.5,Y:=0.5}
                    Widget:=overlay:
                        Slots:=array:
                            overlay_slot:
                                HorizontalAlignment:=horizontal_alignment.Center
                                VerticalAlignment:=vertical_alignment.Center
                                Widget:=overlay:
                                    Slots:=array:
                                        overlay_slot:
                                            Widget:=Slider_Texture_3
                            overlay_slot:
                                HorizontalAlignment:=horizontal_alignment.Center
                                VerticalAlignment:=vertical_alignment.Top
                                Widget:=overlay:
                                    Slots:=array:
                                        overlay_slot:
                                            HorizontalAlignment:=horizontal_alignment.Center
                                            VerticalAlignment:=vertical_alignment.Center
                                            Widget:=texture_block{DefaultDesiredSize:=vector2{X:=XSizeActionUI,Y:=YSizeActionUI},DefaultImage:=Images.EmptyTexture}
                                        <#overlay_slot:
                                            HorizontalAlignment:=horizontal_alignment.Fill
                                            VerticalAlignment:=vertical_alignment.Fill
                                            Widget:=overlay:
                                                Slots:=array:
                                                    overlay_slot:
                                                        HorizontalAlignment:=horizontal_alignment.Fill
                                                        VerticalAlignment:=vertical_alignment.Fill
                                                        Widget:=texture_block{DefaultImage:=Images.Img_LineBackground}#>
                                        overlay_slot:
                                            HorizontalAlignment:=horizontal_alignment.Left
                                            VerticalAlignment:=vertical_alignment.Center
                                            Padding:=margin{Left:=RowLeftPadding,Top:=RowTopPadding,Right:=RowRightPadding,Bottom:=RowBottomPadding}
                                            Widget:=stack_box:
                                                Orientation:=orientation.Horizontal
                                                    Slots:=array:
                                                        stack_box_slot:
                                                            Padding:=margin{Left:=InterColumnLeftPadding,Top:=InterColumnTopPadding,Right:=InterColumnRightPadding,Bottom:=InterColumnBottomPadding}
                                                            HorizontalAlignment:=horizontal_alignment.Left
                                                            VerticalAlignment:=vertical_alignment.Center
                                                            Widget:=Action_Icon_3
                                                        stack_box_slot:
                                                            Padding:=margin{Left:=InterColumnLeftPadding,Top:=InterColumnTopPadding,Right:=InterColumnRightPadding,Bottom:=InterColumnBottomPadding}
                                                            HorizontalAlignment:=horizontal_alignment.Center
                                                            VerticalAlignment:=vertical_alignment.Center
                                                            Widget:=Ammount_TextBlock_3   

                #Moving slider 4
                canvas_slot:
                    Anchors:=anchors{Minimum:=vector2{X:=0.5,Y:=0.5},Maximum:=vector2{X:=0.5,Y:=0.5}}
                    ZOrder:=0
                    SizeToContent:=true
                    Offsets:=margin{Left:=XPositionActionUI_Action,Top:=YPositionActionUI_Action,Right:=0.0,Bottom:=0.0} #PositionX, PositionY, SizeX, SizeY
                    Alignment:=vector2{X:=0.5,Y:=0.5}
                    Widget:=overlay:
                        Slots:=array:
                            overlay_slot:
                                HorizontalAlignment:=horizontal_alignment.Center
                                VerticalAlignment:=vertical_alignment.Center
                                Widget:=overlay:
                                    Slots:=array:
                                        overlay_slot:
                                            Widget:=Slider_Texture_4
                            overlay_slot:
                                HorizontalAlignment:=horizontal_alignment.Center
                                VerticalAlignment:=vertical_alignment.Top
                                Widget:=overlay:
                                    Slots:=array:
                                        overlay_slot:
                                            HorizontalAlignment:=horizontal_alignment.Center
                                            VerticalAlignment:=vertical_alignment.Center
                                            Widget:=texture_block{DefaultDesiredSize:=vector2{X:=XSizeActionUI,Y:=YSizeActionUI},DefaultImage:=Images.EmptyTexture}
                                        overlay_slot:
                                            HorizontalAlignment:=horizontal_alignment.Left
                                            VerticalAlignment:=vertical_alignment.Center
                                            Padding:=margin{Left:=RowLeftPadding,Top:=RowTopPadding,Right:=RowRightPadding,Bottom:=RowBottomPadding}
                                            Widget:=stack_box:
                                                Orientation:=orientation.Horizontal
                                                    Slots:=array:
                                                        stack_box_slot:
                                                            Padding:=margin{Left:=InterColumnLeftPadding,Top:=InterColumnTopPadding,Right:=InterColumnRightPadding,Bottom:=InterColumnBottomPadding}
                                                            HorizontalAlignment:=horizontal_alignment.Left
                                                            VerticalAlignment:=vertical_alignment.Center
                                                            Widget:=Action_Icon_4
                                                        stack_box_slot:
                                                            Padding:=margin{Left:=InterColumnLeftPadding,Top:=InterColumnTopPadding,Right:=InterColumnRightPadding,Bottom:=InterColumnBottomPadding}
                                                            HorizontalAlignment:=horizontal_alignment.Center
                                                            VerticalAlignment:=vertical_alignment.Center
                                                            Widget:=Ammount_TextBlock_4


                                                           
                #Moving slider 5
                canvas_slot:
                    Anchors:=anchors{Minimum:=vector2{X:=0.5,Y:=0.5},Maximum:=vector2{X:=0.5,Y:=0.5}}
                    ZOrder:=0
                    SizeToContent:=true
                    Offsets:=margin{Left:=XPositionActionUI_Action,Top:=YPositionActionUI_Action,Right:=0.0,Bottom:=0.0} #PositionX, PositionY, SizeX, SizeY
                    Alignment:=vector2{X:=0.5,Y:=0.5}
                    Widget:=overlay:
                        Slots:=array:
                            overlay_slot:
                                HorizontalAlignment:=horizontal_alignment.Center
                                VerticalAlignment:=vertical_alignment.Center
                                Widget:=overlay:
                                    Slots:=array:
                                        overlay_slot:
                                            Widget:=Slider_Texture_5
                            overlay_slot:
                                HorizontalAlignment:=horizontal_alignment.Center
                                VerticalAlignment:=vertical_alignment.Top
                                Widget:=overlay:
                                    Slots:=array:
                                        overlay_slot:
                                            HorizontalAlignment:=horizontal_alignment.Center
                                            VerticalAlignment:=vertical_alignment.Center
                                            Widget:=texture_block{DefaultDesiredSize:=vector2{X:=XSizeActionUI,Y:=YSizeActionUI},DefaultImage:=Images.EmptyTexture}
                                        overlay_slot:
                                            HorizontalAlignment:=horizontal_alignment.Left
                                            VerticalAlignment:=vertical_alignment.Center
                                            Padding:=margin{Left:=RowLeftPadding,Top:=RowTopPadding,Right:=RowRightPadding,Bottom:=RowBottomPadding}
                                            Widget:=stack_box:
                                                Orientation:=orientation.Horizontal
                                                    Slots:=array:
                                                        stack_box_slot:
                                                            Padding:=margin{Left:=InterColumnLeftPadding,Top:=InterColumnTopPadding,Right:=InterColumnRightPadding,Bottom:=InterColumnBottomPadding}
                                                            HorizontalAlignment:=horizontal_alignment.Left
                                                            VerticalAlignment:=vertical_alignment.Center
                                                            Widget:=Action_Icon_5
                                                        stack_box_slot:
                                                            Padding:=margin{Left:=InterColumnLeftPadding,Top:=InterColumnTopPadding,Right:=InterColumnRightPadding,Bottom:=InterColumnBottomPadding}
                                                            HorizontalAlignment:=horizontal_alignment.Center
                                                            VerticalAlignment:=vertical_alignment.Center
                                                            Widget:=Ammount_TextBlock_5
                #Moving slider 6
                canvas_slot:
                    Anchors:=anchors{Minimum:=vector2{X:=0.5,Y:=0.5},Maximum:=vector2{X:=0.5,Y:=0.5}}
                    ZOrder:=0
                    SizeToContent:=true
                    Offsets:=margin{Left:=XPositionActionUI_Action,Top:=YPositionActionUI_Action,Right:=0.0,Bottom:=0.0} #PositionX, PositionY, SizeX, SizeY
                    Alignment:=vector2{X:=0.5,Y:=0.5}
                    Widget:=overlay:
                        Slots:=array:
                            overlay_slot:
                                HorizontalAlignment:=horizontal_alignment.Center
                                VerticalAlignment:=vertical_alignment.Center
                                Widget:=overlay:
                                    Slots:=array:
                                        overlay_slot:
                                            Widget:=Slider_Texture_6
                            overlay_slot:
                                HorizontalAlignment:=horizontal_alignment.Center
                                VerticalAlignment:=vertical_alignment.Top
                                Widget:=overlay:
                                    Slots:=array:
                                        overlay_slot:
                                            HorizontalAlignment:=horizontal_alignment.Center
                                            VerticalAlignment:=vertical_alignment.Center
                                            Widget:=texture_block{DefaultDesiredSize:=vector2{X:=XSizeActionUI,Y:=YSizeActionUI},DefaultImage:=Images.EmptyTexture}
                                        overlay_slot:
                                            HorizontalAlignment:=horizontal_alignment.Left
                                            VerticalAlignment:=vertical_alignment.Center
                                            Padding:=margin{Left:=RowLeftPadding,Top:=RowTopPadding,Right:=RowRightPadding,Bottom:=RowBottomPadding}
                                            Widget:=stack_box:
                                                Orientation:=orientation.Horizontal
                                                    Slots:=array:
                                                        stack_box_slot:
                                                            Padding:=margin{Left:=InterColumnLeftPadding,Top:=InterColumnTopPadding,Right:=InterColumnRightPadding,Bottom:=InterColumnBottomPadding}
                                                            HorizontalAlignment:=horizontal_alignment.Left
                                                            VerticalAlignment:=vertical_alignment.Center
                                                            Widget:=Action_Icon_6
                                                        stack_box_slot:
                                                            Padding:=margin{Left:=InterColumnLeftPadding,Top:=InterColumnTopPadding,Right:=InterColumnRightPadding,Bottom:=InterColumnBottomPadding}
                                                            HorizontalAlignment:=horizontal_alignment.Center
                                                            VerticalAlignment:=vertical_alignment.Center
                                                            Widget:=Ammount_TextBlock_6
                #Moving slider 7
                canvas_slot:
                    Anchors:=anchors{Minimum:=vector2{X:=0.5,Y:=0.5},Maximum:=vector2{X:=0.5,Y:=0.5}}
                    ZOrder:=0
                    SizeToContent:=true
                    Offsets:=margin{Left:=XPositionActionUI_Action,Top:=YPositionActionUI_Action,Right:=0.0,Bottom:=0.0} #PositionX, PositionY, SizeX, SizeY
                    Alignment:=vector2{X:=0.5,Y:=0.5}
                    Widget:=overlay:
                        Slots:=array:
                            overlay_slot:
                                HorizontalAlignment:=horizontal_alignment.Center
                                VerticalAlignment:=vertical_alignment.Center
                                Widget:=overlay:
                                    Slots:=array:
                                        overlay_slot:
                                            Widget:=Slider_Texture_7
                            overlay_slot:
                                HorizontalAlignment:=horizontal_alignment.Center
                                VerticalAlignment:=vertical_alignment.Top
                                Widget:=overlay:
                                    Slots:=array:
                                        overlay_slot:
                                            HorizontalAlignment:=horizontal_alignment.Center
                                            VerticalAlignment:=vertical_alignment.Center
                                            Widget:=texture_block{DefaultDesiredSize:=vector2{X:=XSizeActionUI,Y:=YSizeActionUI},DefaultImage:=Images.EmptyTexture}
                                        overlay_slot:
                                            HorizontalAlignment:=horizontal_alignment.Left
                                            VerticalAlignment:=vertical_alignment.Center
                                            Padding:=margin{Left:=RowLeftPadding,Top:=RowTopPadding,Right:=RowRightPadding,Bottom:=RowBottomPadding}
                                            Widget:=stack_box:
                                                Orientation:=orientation.Horizontal
                                                    Slots:=array:
                                                        stack_box_slot:
                                                            Padding:=margin{Left:=InterColumnLeftPadding,Top:=InterColumnTopPadding,Right:=InterColumnRightPadding,Bottom:=InterColumnBottomPadding}
                                                            HorizontalAlignment:=horizontal_alignment.Left
                                                            VerticalAlignment:=vertical_alignment.Center
                                                            Widget:=Action_Icon_7
                                                        stack_box_slot:
                                                            Padding:=margin{Left:=InterColumnLeftPadding,Top:=InterColumnTopPadding,Right:=InterColumnRightPadding,Bottom:=InterColumnBottomPadding}
                                                            HorizontalAlignment:=horizontal_alignment.Center
                                                            VerticalAlignment:=vertical_alignment.Center
                                                            Widget:=Ammount_TextBlock_7
                #Moving slider 8
                canvas_slot:
                    Anchors:=anchors{Minimum:=vector2{X:=0.5,Y:=0.5},Maximum:=vector2{X:=0.5,Y:=0.5}}
                    ZOrder:=0
                    SizeToContent:=true
                    Offsets:=margin{Left:=XPositionActionUI_Action,Top:=YPositionActionUI_Action,Right:=0.0,Bottom:=0.0} #PositionX, PositionY, SizeX, SizeY
                    Alignment:=vector2{X:=0.5,Y:=0.5}
                    Widget:=overlay:
                        Slots:=array:
                            overlay_slot:
                                HorizontalAlignment:=horizontal_alignment.Center
                                VerticalAlignment:=vertical_alignment.Center
                                Widget:=overlay:
                                    Slots:=array:
                                        overlay_slot:
                                            Widget:=Slider_Texture_8
                            overlay_slot:
                                HorizontalAlignment:=horizontal_alignment.Center
                                VerticalAlignment:=vertical_alignment.Top
                                Widget:=overlay:
                                    Slots:=array:
                                        overlay_slot:
                                            HorizontalAlignment:=horizontal_alignment.Center
                                            VerticalAlignment:=vertical_alignment.Center
                                            Widget:=texture_block{DefaultDesiredSize:=vector2{X:=XSizeActionUI,Y:=YSizeActionUI},DefaultImage:=Images.EmptyTexture}
                                        overlay_slot:
                                            HorizontalAlignment:=horizontal_alignment.Left
                                            VerticalAlignment:=vertical_alignment.Center
                                            Padding:=margin{Left:=RowLeftPadding,Top:=RowTopPadding,Right:=RowRightPadding,Bottom:=RowBottomPadding}
                                            Widget:=stack_box:
                                                Orientation:=orientation.Horizontal
                                                    Slots:=array:
                                                        stack_box_slot:
                                                            Padding:=margin{Left:=InterColumnLeftPadding,Top:=InterColumnTopPadding,Right:=InterColumnRightPadding,Bottom:=InterColumnBottomPadding}
                                                            HorizontalAlignment:=horizontal_alignment.Left
                                                            VerticalAlignment:=vertical_alignment.Center
                                                            Widget:=Action_Icon_8
                                                        stack_box_slot:
                                                            Padding:=margin{Left:=InterColumnLeftPadding,Top:=InterColumnTopPadding,Right:=InterColumnRightPadding,Bottom:=InterColumnBottomPadding}
                                                            HorizontalAlignment:=horizontal_alignment.Center
                                                            VerticalAlignment:=vertical_alignment.Center
                                                            Widget:=Ammount_TextBlock_8
                #Moving slider 9
                canvas_slot:
                    Anchors:=anchors{Minimum:=vector2{X:=0.5,Y:=0.5},Maximum:=vector2{X:=0.5,Y:=0.5}}
                    ZOrder:=0
                    SizeToContent:=true
                    Offsets:=margin{Left:=XPositionActionUI_Action,Top:=YPositionActionUI_Action,Right:=0.0,Bottom:=0.0} #PositionX, PositionY, SizeX, SizeY
                    Alignment:=vector2{X:=0.5,Y:=0.5}
                    Widget:=overlay:
                        Slots:=array:
                            overlay_slot:
                                HorizontalAlignment:=horizontal_alignment.Center
                                VerticalAlignment:=vertical_alignment.Center
                                Widget:=overlay:
                                    Slots:=array:
                                        overlay_slot:
                                            Widget:=Slider_Texture_9
                            overlay_slot:
                                HorizontalAlignment:=horizontal_alignment.Center
                                VerticalAlignment:=vertical_alignment.Top
                                Widget:=overlay:
                                    Slots:=array:
                                        overlay_slot:
                                            HorizontalAlignment:=horizontal_alignment.Center
                                            VerticalAlignment:=vertical_alignment.Center
                                            Widget:=texture_block{DefaultDesiredSize:=vector2{X:=XSizeActionUI,Y:=YSizeActionUI},DefaultImage:=Images.EmptyTexture}
                                        overlay_slot:
                                            HorizontalAlignment:=horizontal_alignment.Left
                                            VerticalAlignment:=vertical_alignment.Center
                                            Padding:=margin{Left:=RowLeftPadding,Top:=RowTopPadding,Right:=RowRightPadding,Bottom:=RowBottomPadding}
                                            Widget:=stack_box:
                                                Orientation:=orientation.Horizontal
                                                    Slots:=array:
                                                        stack_box_slot:
                                                            Padding:=margin{Left:=InterColumnLeftPadding,Top:=InterColumnTopPadding,Right:=InterColumnRightPadding,Bottom:=InterColumnBottomPadding}
                                                            HorizontalAlignment:=horizontal_alignment.Left
                                                            VerticalAlignment:=vertical_alignment.Center
                                                            Widget:=Action_Icon_9
                                                        stack_box_slot:
                                                            Padding:=margin{Left:=InterColumnLeftPadding,Top:=InterColumnTopPadding,Right:=InterColumnRightPadding,Bottom:=InterColumnBottomPadding}
                                                            HorizontalAlignment:=horizontal_alignment.Center
                                                            VerticalAlignment:=vertical_alignment.Center
                                                            Widget:=Ammount_TextBlock_9
                #Moving slider 10
                canvas_slot:
                    Anchors:=anchors{Minimum:=vector2{X:=0.5,Y:=0.5},Maximum:=vector2{X:=0.5,Y:=0.5}}
                    ZOrder:=0
                    SizeToContent:=true
                    Offsets:=margin{Left:=XPositionActionUI_Action,Top:=YPositionActionUI_Action,Right:=0.0,Bottom:=0.0} #PositionX, PositionY, SizeX, SizeY
                    Alignment:=vector2{X:=0.5,Y:=0.5}
                    Widget:=overlay:
                        Slots:=array:
                            overlay_slot:
                                HorizontalAlignment:=horizontal_alignment.Center
                                VerticalAlignment:=vertical_alignment.Center
                                Widget:=overlay:
                                    Slots:=array:
                                        overlay_slot:
                                            Widget:=Slider_Texture_10
                            overlay_slot:
                                HorizontalAlignment:=horizontal_alignment.Center
                                VerticalAlignment:=vertical_alignment.Top
                                Widget:=overlay:
                                    Slots:=array:
                                        overlay_slot:
                                            HorizontalAlignment:=horizontal_alignment.Center
                                            VerticalAlignment:=vertical_alignment.Center
                                            Widget:=texture_block{DefaultDesiredSize:=vector2{X:=XSizeActionUI,Y:=YSizeActionUI},DefaultImage:=Images.EmptyTexture}
                                        overlay_slot:
                                            HorizontalAlignment:=horizontal_alignment.Left
                                            VerticalAlignment:=vertical_alignment.Center
                                            Padding:=margin{Left:=RowLeftPadding,Top:=RowTopPadding,Right:=RowRightPadding,Bottom:=RowBottomPadding}
                                            Widget:=stack_box:
                                                Orientation:=orientation.Horizontal
                                                    Slots:=array:
                                                        stack_box_slot:
                                                            Padding:=margin{Left:=InterColumnLeftPadding,Top:=InterColumnTopPadding,Right:=InterColumnRightPadding,Bottom:=InterColumnBottomPadding}
                                                            HorizontalAlignment:=horizontal_alignment.Left
                                                            VerticalAlignment:=vertical_alignment.Center
                                                            Widget:=Action_Icon_10
                                                        stack_box_slot:
                                                            Padding:=margin{Left:=InterColumnLeftPadding,Top:=InterColumnTopPadding,Right:=InterColumnRightPadding,Bottom:=InterColumnBottomPadding}
                                                            HorizontalAlignment:=horizontal_alignment.Center
                                                            VerticalAlignment:=vertical_alignment.Center
                                                            Widget:=Ammount_TextBlock_10
                #Moving slider 11
                canvas_slot:
                    Anchors:=anchors{Minimum:=vector2{X:=0.5,Y:=0.5},Maximum:=vector2{X:=0.5,Y:=0.5}}
                    ZOrder:=0
                    SizeToContent:=true
                    Offsets:=margin{Left:=XPositionActionUI_Action,Top:=YPositionActionUI_Action,Right:=0.0,Bottom:=0.0} #PositionX, PositionY, SizeX, SizeY
                    Alignment:=vector2{X:=0.5,Y:=0.5}
                    Widget:=overlay:
                        Slots:=array:
                            overlay_slot:
                                HorizontalAlignment:=horizontal_alignment.Center
                                VerticalAlignment:=vertical_alignment.Center
                                Widget:=overlay:
                                    Slots:=array:
                                        overlay_slot:
                                            Widget:=Slider_Texture_11
                            overlay_slot:
                                HorizontalAlignment:=horizontal_alignment.Center
                                VerticalAlignment:=vertical_alignment.Top
                                Widget:=overlay:
                                    Slots:=array:
                                        overlay_slot:
                                            HorizontalAlignment:=horizontal_alignment.Center
                                            VerticalAlignment:=vertical_alignment.Center
                                            Widget:=texture_block{DefaultDesiredSize:=vector2{X:=XSizeActionUI,Y:=YSizeActionUI},DefaultImage:=Images.EmptyTexture}
                                        overlay_slot:
                                            HorizontalAlignment:=horizontal_alignment.Left
                                            VerticalAlignment:=vertical_alignment.Center
                                            Padding:=margin{Left:=RowLeftPadding,Top:=RowTopPadding,Right:=RowRightPadding,Bottom:=RowBottomPadding}
                                            Widget:=stack_box:
                                                Orientation:=orientation.Horizontal
                                                    Slots:=array:
                                                        stack_box_slot:
                                                            Padding:=margin{Left:=InterColumnLeftPadding,Top:=InterColumnTopPadding,Right:=InterColumnRightPadding,Bottom:=InterColumnBottomPadding}
                                                            HorizontalAlignment:=horizontal_alignment.Left
                                                            VerticalAlignment:=vertical_alignment.Center
                                                            Widget:=Action_Icon_11
                                                        stack_box_slot:
                                                            Padding:=margin{Left:=InterColumnLeftPadding,Top:=InterColumnTopPadding,Right:=InterColumnRightPadding,Bottom:=InterColumnBottomPadding}
                                                            HorizontalAlignment:=horizontal_alignment.Center
                                                            VerticalAlignment:=vertical_alignment.Center
                                                            Widget:=Ammount_TextBlock_11
                #Moving slider 12
                canvas_slot:
                    Anchors:=anchors{Minimum:=vector2{X:=0.5,Y:=0.5},Maximum:=vector2{X:=0.5,Y:=0.5}}
                    ZOrder:=0
                    SizeToContent:=true
                    Offsets:=margin{Left:=XPositionActionUI_Action,Top:=YPositionActionUI_Action,Right:=0.0,Bottom:=0.0} #PositionX, PositionY, SizeX, SizeY
                    Alignment:=vector2{X:=0.5,Y:=0.5}
                    Widget:=overlay:
                        Slots:=array:
                            overlay_slot:
                                HorizontalAlignment:=horizontal_alignment.Center
                                VerticalAlignment:=vertical_alignment.Center
                                Widget:=overlay:
                                    Slots:=array:
                                        overlay_slot:
                                            Widget:=Slider_Texture_12
                            overlay_slot:
                                HorizontalAlignment:=horizontal_alignment.Center
                                VerticalAlignment:=vertical_alignment.Top
                                Widget:=overlay:
                                    Slots:=array:
                                        overlay_slot:
                                            HorizontalAlignment:=horizontal_alignment.Center
                                            VerticalAlignment:=vertical_alignment.Center
                                            Widget:=texture_block{DefaultDesiredSize:=vector2{X:=XSizeActionUI,Y:=YSizeActionUI},DefaultImage:=Images.EmptyTexture}
                                        overlay_slot:
                                            HorizontalAlignment:=horizontal_alignment.Left
                                            VerticalAlignment:=vertical_alignment.Center
                                            Padding:=margin{Left:=RowLeftPadding,Top:=RowTopPadding,Right:=RowRightPadding,Bottom:=RowBottomPadding}
                                            Widget:=stack_box:
                                                Orientation:=orientation.Horizontal
                                                    Slots:=array:
                                                        stack_box_slot:
                                                            Padding:=margin{Left:=InterColumnLeftPadding,Top:=InterColumnTopPadding,Right:=InterColumnRightPadding,Bottom:=InterColumnBottomPadding}
                                                            HorizontalAlignment:=horizontal_alignment.Left
                                                            VerticalAlignment:=vertical_alignment.Center
                                                            Widget:=Action_Icon_12
                                                        stack_box_slot:
                                                            Padding:=margin{Left:=InterColumnLeftPadding,Top:=InterColumnTopPadding,Right:=InterColumnRightPadding,Bottom:=InterColumnBottomPadding}
                                                            HorizontalAlignment:=horizontal_alignment.Center
                                                            VerticalAlignment:=vertical_alignment.Center
                                                            Widget:=Ammount_TextBlock_12
                #Moving slider 13
                canvas_slot:
                    Anchors:=anchors{Minimum:=vector2{X:=0.5,Y:=0.5},Maximum:=vector2{X:=0.5,Y:=0.5}}
                    ZOrder:=0
                    SizeToContent:=true
                    Offsets:=margin{Left:=XPositionActionUI_Action,Top:=YPositionActionUI_Action,Right:=0.0,Bottom:=0.0} #PositionX, PositionY, SizeX, SizeY
                    Alignment:=vector2{X:=0.5,Y:=0.5}
                    Widget:=overlay:
                        Slots:=array:
                            overlay_slot:
                                HorizontalAlignment:=horizontal_alignment.Center
                                VerticalAlignment:=vertical_alignment.Center
                                Widget:=overlay:
                                    Slots:=array:
                                        overlay_slot:
                                            Widget:=Slider_Texture_13
                            overlay_slot:
                                HorizontalAlignment:=horizontal_alignment.Center
                                VerticalAlignment:=vertical_alignment.Top
                                Widget:=overlay:
                                    Slots:=array:
                                        overlay_slot:
                                            HorizontalAlignment:=horizontal_alignment.Center
                                            VerticalAlignment:=vertical_alignment.Center
                                            Widget:=texture_block{DefaultDesiredSize:=vector2{X:=XSizeActionUI,Y:=YSizeActionUI},DefaultImage:=Images.EmptyTexture}
                                        overlay_slot:
                                            HorizontalAlignment:=horizontal_alignment.Left
                                            VerticalAlignment:=vertical_alignment.Center
                                            Padding:=margin{Left:=RowLeftPadding,Top:=RowTopPadding,Right:=RowRightPadding,Bottom:=RowBottomPadding}
                                            Widget:=stack_box:
                                                Orientation:=orientation.Horizontal
                                                    Slots:=array:
                                                        stack_box_slot:
                                                            Padding:=margin{Left:=InterColumnLeftPadding,Top:=InterColumnTopPadding,Right:=InterColumnRightPadding,Bottom:=InterColumnBottomPadding}
                                                            HorizontalAlignment:=horizontal_alignment.Left
                                                            VerticalAlignment:=vertical_alignment.Center
                                                            Widget:=Action_Icon_13
                                                        stack_box_slot:
                                                            Padding:=margin{Left:=InterColumnLeftPadding,Top:=InterColumnTopPadding,Right:=InterColumnRightPadding,Bottom:=InterColumnBottomPadding}
                                                            HorizontalAlignment:=horizontal_alignment.Center
                                                            VerticalAlignment:=vertical_alignment.Center
                                                            Widget:=Ammount_TextBlock_13
                #Moving slider 14
                canvas_slot:
                    Anchors:=anchors{Minimum:=vector2{X:=0.5,Y:=0.5},Maximum:=vector2{X:=0.5,Y:=0.5}}
                    ZOrder:=0
                    SizeToContent:=true
                    Offsets:=margin{Left:=XPositionActionUI_Action,Top:=YPositionActionUI_Action,Right:=0.0,Bottom:=0.0} #PositionX, PositionY, SizeX, SizeY
                    Alignment:=vector2{X:=0.5,Y:=0.5}
                    Widget:=overlay:
                        Slots:=array:
                            overlay_slot:
                                HorizontalAlignment:=horizontal_alignment.Center
                                VerticalAlignment:=vertical_alignment.Center
                                Widget:=overlay:
                                    Slots:=array:
                                        overlay_slot:
                                            Widget:=Slider_Texture_14
                            overlay_slot:
                                HorizontalAlignment:=horizontal_alignment.Center
                                VerticalAlignment:=vertical_alignment.Top
                                Widget:=overlay:
                                    Slots:=array:
                                        overlay_slot:
                                            HorizontalAlignment:=horizontal_alignment.Center
                                            VerticalAlignment:=vertical_alignment.Center
                                            Widget:=texture_block{DefaultDesiredSize:=vector2{X:=XSizeActionUI,Y:=YSizeActionUI},DefaultImage:=Images.EmptyTexture}
                                        overlay_slot:
                                            HorizontalAlignment:=horizontal_alignment.Left
                                            VerticalAlignment:=vertical_alignment.Center
                                            Padding:=margin{Left:=RowLeftPadding,Top:=RowTopPadding,Right:=RowRightPadding,Bottom:=RowBottomPadding}
                                            Widget:=stack_box:
                                                Orientation:=orientation.Horizontal
                                                    Slots:=array:
                                                        stack_box_slot:
                                                            Padding:=margin{Left:=InterColumnLeftPadding,Top:=InterColumnTopPadding,Right:=InterColumnRightPadding,Bottom:=InterColumnBottomPadding}
                                                            HorizontalAlignment:=horizontal_alignment.Left
                                                            VerticalAlignment:=vertical_alignment.Center
                                                            Widget:=Action_Icon_14
                                                        stack_box_slot:
                                                            Padding:=margin{Left:=InterColumnLeftPadding,Top:=InterColumnTopPadding,Right:=InterColumnRightPadding,Bottom:=InterColumnBottomPadding}
                                                            HorizontalAlignment:=horizontal_alignment.Center
                                                            VerticalAlignment:=vertical_alignment.Center
                                                            Widget:=Ammount_TextBlock_14
                #Moving slider 15
                canvas_slot:
                    Anchors:=anchors{Minimum:=vector2{X:=0.5,Y:=0.5},Maximum:=vector2{X:=0.5,Y:=0.5}}
                    ZOrder:=0
                    SizeToContent:=true
                    Offsets:=margin{Left:=XPositionActionUI_Action,Top:=YPositionActionUI_Action,Right:=0.0,Bottom:=0.0} #PositionX, PositionY, SizeX, SizeY
                    Alignment:=vector2{X:=0.5,Y:=0.5}
                    Widget:=overlay:
                        Slots:=array:
                            overlay_slot:
                                HorizontalAlignment:=horizontal_alignment.Center
                                VerticalAlignment:=vertical_alignment.Center
                                Widget:=overlay:
                                    Slots:=array:
                                        overlay_slot:
                                            Widget:=Slider_Texture_15
                            overlay_slot:
                                HorizontalAlignment:=horizontal_alignment.Center
                                VerticalAlignment:=vertical_alignment.Top
                                Widget:=overlay:
                                    Slots:=array:
                                        overlay_slot:
                                            HorizontalAlignment:=horizontal_alignment.Center
                                            VerticalAlignment:=vertical_alignment.Center
                                            Widget:=texture_block{DefaultDesiredSize:=vector2{X:=XSizeActionUI,Y:=YSizeActionUI},DefaultImage:=Images.EmptyTexture}
                                        overlay_slot:
                                            HorizontalAlignment:=horizontal_alignment.Left
                                            VerticalAlignment:=vertical_alignment.Center
                                            Padding:=margin{Left:=RowLeftPadding,Top:=RowTopPadding,Right:=RowRightPadding,Bottom:=RowBottomPadding}
                                            Widget:=stack_box:
                                                Orientation:=orientation.Horizontal
                                                    Slots:=array:
                                                        stack_box_slot:
                                                            Padding:=margin{Left:=InterColumnLeftPadding,Top:=InterColumnTopPadding,Right:=InterColumnRightPadding,Bottom:=InterColumnBottomPadding}
                                                            HorizontalAlignment:=horizontal_alignment.Left
                                                            VerticalAlignment:=vertical_alignment.Center
                                                            Widget:=Action_Icon_15
                                                        stack_box_slot:
                                                            Padding:=margin{Left:=InterColumnLeftPadding,Top:=InterColumnTopPadding,Right:=InterColumnRightPadding,Bottom:=InterColumnBottomPadding}
                                                            HorizontalAlignment:=horizontal_alignment.Center
                                                            VerticalAlignment:=vertical_alignment.Center
                                                            Widget:=Ammount_TextBlock_15
                #Moving slider 16
                canvas_slot:
                    Anchors:=anchors{Minimum:=vector2{X:=0.5,Y:=0.5},Maximum:=vector2{X:=0.5,Y:=0.5}}
                    ZOrder:=0
                    SizeToContent:=true
                    Offsets:=margin{Left:=XPositionActionUI_Action,Top:=YPositionActionUI_Action,Right:=0.0,Bottom:=0.0} #PositionX, PositionY, SizeX, SizeY
                    Alignment:=vector2{X:=0.5,Y:=0.5}
                    Widget:=overlay:
                        Slots:=array:
                            overlay_slot:
                                HorizontalAlignment:=horizontal_alignment.Center
                                VerticalAlignment:=vertical_alignment.Center
                                Widget:=overlay:
                                    Slots:=array:
                                        overlay_slot:
                                            Widget:=Slider_Texture_16
                            overlay_slot:
                                HorizontalAlignment:=horizontal_alignment.Center
                                VerticalAlignment:=vertical_alignment.Top
                                Widget:=overlay:
                                    Slots:=array:
                                        overlay_slot:
                                            HorizontalAlignment:=horizontal_alignment.Center
                                            VerticalAlignment:=vertical_alignment.Center
                                            Widget:=texture_block{DefaultDesiredSize:=vector2{X:=XSizeActionUI,Y:=YSizeActionUI},DefaultImage:=Images.EmptyTexture}
                                        overlay_slot:
                                            HorizontalAlignment:=horizontal_alignment.Left
                                            VerticalAlignment:=vertical_alignment.Center
                                            Padding:=margin{Left:=RowLeftPadding,Top:=RowTopPadding,Right:=RowRightPadding,Bottom:=RowBottomPadding}
                                            Widget:=stack_box:
                                                Orientation:=orientation.Horizontal
                                                    Slots:=array:
                                                        stack_box_slot:
                                                            Padding:=margin{Left:=InterColumnLeftPadding,Top:=InterColumnTopPadding,Right:=InterColumnRightPadding,Bottom:=InterColumnBottomPadding}
                                                            HorizontalAlignment:=horizontal_alignment.Left
                                                            VerticalAlignment:=vertical_alignment.Center
                                                            Widget:=Action_Icon_16
                                                        stack_box_slot:
                                                            Padding:=margin{Left:=InterColumnLeftPadding,Top:=InterColumnTopPadding,Right:=InterColumnRightPadding,Bottom:=InterColumnBottomPadding}
                                                            HorizontalAlignment:=horizontal_alignment.Center
                                                            VerticalAlignment:=vertical_alignment.Center
                                                            Widget:=Ammount_TextBlock_16
                                                                
                #Water resource ammount
                canvas_slot :
                    Anchors:=anchors{Minimum:=vector2{X:=0.5,Y:=0.5},Maximum:=vector2{X:=0.5,Y:=0.5}}
                    ZOrder:=0
                    SizeToContent:=true
                    Offsets:=margin{Left:=XPositionUI_WaterRes,Top:=YPositionUI_WaterRes,Right:=0.0,Bottom:=0.0} #PositionX, PositionY, SizeX, SizeY
                    Alignment:=vector2{X:=0.5,Y:=0.5}
                    Widget:=overlay:
                        Slots:=array:
                            overlay_slot:
                                HorizontalAlignment:=horizontal_alignment.Center
                                VerticalAlignment:=vertical_alignment.Center
                                Widget:=overlay:
                                    Slots:=array:
                                        overlay_slot:
                                            Widget:=WaterResource_Slider
                            overlay_slot:
                                HorizontalAlignment:=horizontal_alignment.Center
                                VerticalAlignment:=vertical_alignment.Top
                                Widget:=overlay:
                                    Slots:=array:
                                        overlay_slot:
                                            HorizontalAlignment:=horizontal_alignment.Center
                                            VerticalAlignment:=vertical_alignment.Center
                                            Widget:=texture_block{DefaultDesiredSize:=vector2{X:=XSizeActionUI,Y:=YSizeActionUI},DefaultImage:=Images.EmptyTexture}
                                        overlay_slot:
                                            HorizontalAlignment:=horizontal_alignment.Fill
                                            VerticalAlignment:=vertical_alignment.Fill
                                            Widget:=overlay:
                                                Slots:=array:
                                                    overlay_slot:
                                                        HorizontalAlignment:=horizontal_alignment.Fill
                                                        VerticalAlignment:=vertical_alignment.Fill
                                                        Padding:=margin{Left:=ActionImgPadding,Top:=ActionImgPadding,Right:=ActionImgPadding,Bottom:=ActionImgPadding}
                                                        Widget:=texture_block{DefaultDesiredSize:=vector2{X:=XSizeActionUI,Y:=YSizeActionUI},DefaultTint:=WaterColor,DefaultImage:=Images.Img_LineBackground}
                                                    overlay_slot:
                                                        HorizontalAlignment:=horizontal_alignment.Fill
                                                        VerticalAlignment:=vertical_alignment.Fill
                                                        Widget:=texture_block{DefaultDesiredSize:=vector2{X:=XSizeActionUI,Y:=YSizeActionUI},DefaultImage:=Images.Img_LineBackground}
                                        overlay_slot:
                                            HorizontalAlignment:=horizontal_alignment.Left
                                            VerticalAlignment:=vertical_alignment.Center
                                            Padding:=margin{Left:=RowLeftPadding,Top:=RowTopPadding,Right:=RowRightPadding,Bottom:=RowBottomPadding}
                                            Widget:=stack_box:
                                                Orientation:=orientation.Horizontal
                                                    Slots:=array:
                                                        stack_box_slot:
                                                            Padding:=margin{Left:=InterColumnLeftPadding,Top:=InterColumnTopPadding,Right:=InterColumnRightPadding,Bottom:=InterColumnBottomPadding}
                                                            HorizontalAlignment:=horizontal_alignment.Left
                                                            VerticalAlignment:=vertical_alignment.Center
                                                            Widget:=WaterResource_Icon
                                                        stack_box_slot:
                                                            Padding:=margin{Left:=InterColumnLeftPadding,Top:=InterColumnTopPadding,Right:=InterColumnRightPadding,Bottom:=InterColumnBottomPadding}
                                                            HorizontalAlignment:=horizontal_alignment.Center
                                                            VerticalAlignment:=vertical_alignment.Center
                                                            Widget:=WaterResourceAmmount_TextBlock 

                #Flour resource ammount
                canvas_slot :
                    Anchors:=anchors{Minimum:=vector2{X:=0.5,Y:=0.5},Maximum:=vector2{X:=0.5,Y:=0.5}}
                    ZOrder:=0
                    SizeToContent:=true
                    Offsets:=margin{Left:=XPositionUI_FlourRes,Top:=YPositionUI_FlourRes,Right:=0.0,Bottom:=0.0} #PositionX, PositionY, SizeX, SizeY
                    Alignment:=vector2{X:=0.5,Y:=0.5}
                    Widget:=overlay:
                        Slots:=array:
                            overlay_slot:
                                HorizontalAlignment:=horizontal_alignment.Center
                                VerticalAlignment:=vertical_alignment.Center
                                Widget:=overlay:
                                    Slots:=array:
                                        overlay_slot:
                                            Widget:=FlourResource_Slider
                            overlay_slot:
                                HorizontalAlignment:=horizontal_alignment.Center
                                VerticalAlignment:=vertical_alignment.Top
                                Widget:=overlay:
                                    Slots:=array:
                                        overlay_slot:
                                            HorizontalAlignment:=horizontal_alignment.Center
                                            VerticalAlignment:=vertical_alignment.Center
                                            Widget:=texture_block{DefaultDesiredSize:=vector2{X:=XSizeActionUI,Y:=YSizeActionUI},DefaultImage:=Images.EmptyTexture}
                                        overlay_slot:
                                            HorizontalAlignment:=horizontal_alignment.Fill
                                            VerticalAlignment:=vertical_alignment.Fill
                                            Widget:=overlay:
                                                Slots:=array:
                                                    overlay_slot:
                                                        HorizontalAlignment:=horizontal_alignment.Fill
                                                        VerticalAlignment:=vertical_alignment.Fill
                                                        Padding:=margin{Left:=ActionImgPadding,Top:=ActionImgPadding,Right:=ActionImgPadding,Bottom:=ActionImgPadding}
                                                        Widget:=texture_block{DefaultDesiredSize:=vector2{X:=XSizeActionUI,Y:=YSizeActionUI},DefaultTint:=FlourColor,DefaultImage:=Images.Img_LineBackground}
                                                    overlay_slot:
                                                        HorizontalAlignment:=horizontal_alignment.Fill
                                                        VerticalAlignment:=vertical_alignment.Fill
                                                        Widget:=texture_block{DefaultDesiredSize:=vector2{X:=XSizeActionUI,Y:=YSizeActionUI},DefaultImage:=Images.Img_LineBackground}
                                        overlay_slot:
                                            HorizontalAlignment:=horizontal_alignment.Left
                                            VerticalAlignment:=vertical_alignment.Center
                                            Padding:=margin{Left:=RowLeftPadding,Top:=RowTopPadding,Right:=RowRightPadding,Bottom:=RowBottomPadding}
                                            Widget:=stack_box:
                                                Orientation:=orientation.Horizontal
                                                    Slots:=array:
                                                        stack_box_slot:
                                                            Padding:=margin{Left:=InterColumnLeftPadding,Top:=InterColumnTopPadding,Right:=InterColumnRightPadding,Bottom:=InterColumnBottomPadding}
                                                            HorizontalAlignment:=horizontal_alignment.Left
                                                            VerticalAlignment:=vertical_alignment.Center
                                                            Widget:=FlourResource_Icon
                                                        stack_box_slot:
                                                            Padding:=margin{Left:=InterColumnLeftPadding,Top:=InterColumnTopPadding,Right:=InterColumnRightPadding,Bottom:=InterColumnBottomPadding}
                                                            HorizontalAlignment:=horizontal_alignment.Center
                                                            VerticalAlignment:=vertical_alignment.Center
                                                            Widget:=FlourResourceAmmount_TextBlock 

                #Cheese resource ammount
                canvas_slot :
                    Anchors:=anchors{Minimum:=vector2{X:=0.5,Y:=0.5},Maximum:=vector2{X:=0.5,Y:=0.5}}
                    ZOrder:=0
                    SizeToContent:=true
                    Offsets:=margin{Left:=XPositionUI_CheeseRes,Top:=YPositionUI_CheeseRes,Right:=0.0,Bottom:=0.0} #PositionX, PositionY, SizeX, SizeY
                    Alignment:=vector2{X:=0.5,Y:=0.5}
                    Widget:=overlay:
                        Slots:=array:
                            overlay_slot:
                                HorizontalAlignment:=horizontal_alignment.Center
                                VerticalAlignment:=vertical_alignment.Center
                                Widget:=overlay:
                                    Slots:=array:
                                        overlay_slot:
                                            Widget:=CheeseResource_Slider
                            overlay_slot:
                                HorizontalAlignment:=horizontal_alignment.Center
                                VerticalAlignment:=vertical_alignment.Top
                                Widget:=overlay:
                                    Slots:=array:
                                        overlay_slot:
                                            HorizontalAlignment:=horizontal_alignment.Center
                                            VerticalAlignment:=vertical_alignment.Center
                                            Widget:=texture_block{DefaultDesiredSize:=vector2{X:=XSizeActionUI,Y:=YSizeActionUI},DefaultImage:=Images.EmptyTexture}
                                        overlay_slot:
                                            HorizontalAlignment:=horizontal_alignment.Fill
                                            VerticalAlignment:=vertical_alignment.Fill
                                            Widget:=overlay:
                                                Slots:=array:
                                                    overlay_slot:
                                                        HorizontalAlignment:=horizontal_alignment.Fill
                                                        VerticalAlignment:=vertical_alignment.Fill
                                                        Padding:=margin{Left:=ActionImgPadding,Top:=ActionImgPadding,Right:=ActionImgPadding,Bottom:=ActionImgPadding}
                                                        Widget:=texture_block{DefaultDesiredSize:=vector2{X:=XSizeActionUI,Y:=YSizeActionUI},DefaultTint:=CheeseColor,DefaultImage:=Images.Img_LineBackground}
                                                    overlay_slot:
                                                        HorizontalAlignment:=horizontal_alignment.Fill
                                                        VerticalAlignment:=vertical_alignment.Fill
                                                        Widget:=texture_block{DefaultDesiredSize:=vector2{X:=XSizeActionUI,Y:=YSizeActionUI},DefaultImage:=Images.Img_LineBackground}
                                        overlay_slot:
                                            HorizontalAlignment:=horizontal_alignment.Left
                                            VerticalAlignment:=vertical_alignment.Center
                                            Padding:=margin{Left:=RowLeftPadding,Top:=RowTopPadding,Right:=RowRightPadding,Bottom:=RowBottomPadding}
                                            Widget:=stack_box:
                                                Orientation:=orientation.Horizontal
                                                    Slots:=array:
                                                        stack_box_slot:
                                                            Padding:=margin{Left:=InterColumnLeftPadding,Top:=InterColumnTopPadding,Right:=InterColumnRightPadding,Bottom:=InterColumnBottomPadding}
                                                            HorizontalAlignment:=horizontal_alignment.Left
                                                            VerticalAlignment:=vertical_alignment.Center
                                                            Widget:=CheeseResource_Icon
                                                        stack_box_slot:
                                                            Padding:=margin{Left:=InterColumnLeftPadding,Top:=InterColumnTopPadding,Right:=InterColumnRightPadding,Bottom:=InterColumnBottomPadding}
                                                            HorizontalAlignment:=horizontal_alignment.Center
                                                            VerticalAlignment:=vertical_alignment.Center
                                                            Widget:=CheeseResourceAmmount_TextBlock 


                #Pimentos resource ammount
                canvas_slot :
                    Anchors:=anchors{Minimum:=vector2{X:=0.5,Y:=0.5},Maximum:=vector2{X:=0.5,Y:=0.5}}
                    ZOrder:=0
                    SizeToContent:=true
                    Offsets:=margin{Left:=XPositionUI_PimentosRes,Top:=YPositionUI_PimentosRes,Right:=0.0,Bottom:=0.0} #PositionX, PositionY, SizeX, SizeY
                    Alignment:=vector2{X:=0.5,Y:=0.5}
                    Widget:=overlay:
                        Slots:=array:
                            overlay_slot:
                                HorizontalAlignment:=horizontal_alignment.Center
                                VerticalAlignment:=vertical_alignment.Center
                                Widget:=overlay:
                                    Slots:=array:
                                        overlay_slot:
                                            Widget:=PimentosResource_Slider
                            overlay_slot:
                                HorizontalAlignment:=horizontal_alignment.Center
                                VerticalAlignment:=vertical_alignment.Top
                                Widget:=overlay:
                                    Slots:=array:
                                        overlay_slot:
                                            HorizontalAlignment:=horizontal_alignment.Center
                                            VerticalAlignment:=vertical_alignment.Center
                                            Widget:=texture_block{DefaultDesiredSize:=vector2{X:=XSizeActionUI,Y:=YSizeActionUI},DefaultImage:=Images.EmptyTexture}
                                        overlay_slot:
                                            HorizontalAlignment:=horizontal_alignment.Fill
                                            VerticalAlignment:=vertical_alignment.Fill
                                            Widget:=overlay:
                                                Slots:=array:
                                                    overlay_slot:
                                                        HorizontalAlignment:=horizontal_alignment.Fill
                                                        VerticalAlignment:=vertical_alignment.Fill
                                                        Padding:=margin{Left:=ActionImgPadding,Top:=ActionImgPadding,Right:=ActionImgPadding,Bottom:=ActionImgPadding}
                                                        Widget:=texture_block{DefaultDesiredSize:=vector2{X:=XSizeActionUI,Y:=YSizeActionUI},DefaultTint:=PimentoColor,DefaultImage:=Images.Img_LineBackground}
                                                    overlay_slot:
                                                        HorizontalAlignment:=horizontal_alignment.Fill
                                                        VerticalAlignment:=vertical_alignment.Fill
                                                        Widget:=texture_block{DefaultDesiredSize:=vector2{X:=XSizeActionUI,Y:=YSizeActionUI},DefaultImage:=Images.Img_LineBackground}
                                        overlay_slot:
                                            HorizontalAlignment:=horizontal_alignment.Left
                                            VerticalAlignment:=vertical_alignment.Center
                                            Padding:=margin{Left:=RowLeftPadding,Top:=RowTopPadding,Right:=RowRightPadding,Bottom:=RowBottomPadding}
                                            Widget:=stack_box:
                                                Orientation:=orientation.Horizontal
                                                    Slots:=array:
                                                        stack_box_slot:
                                                            Padding:=margin{Left:=InterColumnLeftPadding,Top:=InterColumnTopPadding,Right:=InterColumnRightPadding,Bottom:=InterColumnBottomPadding}
                                                            HorizontalAlignment:=horizontal_alignment.Left
                                                            VerticalAlignment:=vertical_alignment.Center
                                                            Widget:=PimentosResource_Icon
                                                        stack_box_slot:
                                                            Padding:=margin{Left:=InterColumnLeftPadding,Top:=InterColumnTopPadding,Right:=InterColumnRightPadding,Bottom:=InterColumnBottomPadding}
                                                            HorizontalAlignment:=horizontal_alignment.Center
                                                            VerticalAlignment:=vertical_alignment.Center
                                                            Widget:=PimentosResourceAmmount_TextBlock 
                #Coal resource ammount
                canvas_slot :
                    Anchors:=anchors{Minimum:=vector2{X:=0.5,Y:=0.5},Maximum:=vector2{X:=0.5,Y:=0.5}}
                    ZOrder:=0
                    SizeToContent:=true
                    Offsets:=margin{Left:=XPositionUI_CoalRes,Top:=YPositionUI_CoalRes,Right:=0.0,Bottom:=0.0} #PositionX, PositionY, SizeX, SizeY
                    Alignment:=vector2{X:=0.5,Y:=0.5}
                    Widget:=overlay:
                        Slots:=array:
                            overlay_slot:
                                HorizontalAlignment:=horizontal_alignment.Center
                                VerticalAlignment:=vertical_alignment.Center
                                Widget:=overlay:
                                    Slots:=array:
                                        overlay_slot:
                                            Widget:=CoalResource_Slider
                            overlay_slot:
                                HorizontalAlignment:=horizontal_alignment.Center
                                VerticalAlignment:=vertical_alignment.Top
                                Widget:=overlay:
                                    Slots:=array:
                                        overlay_slot:
                                            HorizontalAlignment:=horizontal_alignment.Center
                                            VerticalAlignment:=vertical_alignment.Center
                                            Widget:=texture_block{DefaultDesiredSize:=vector2{X:=XSizeActionUI,Y:=YSizeActionUI},DefaultImage:=Images.EmptyTexture}
                                        overlay_slot:
                                            HorizontalAlignment:=horizontal_alignment.Fill
                                            VerticalAlignment:=vertical_alignment.Fill
                                            Widget:=overlay:
                                                Slots:=array:
                                                    overlay_slot:
                                                        HorizontalAlignment:=horizontal_alignment.Fill
                                                        VerticalAlignment:=vertical_alignment.Fill
                                                        Padding:=margin{Left:=ActionImgPadding,Top:=ActionImgPadding,Right:=ActionImgPadding,Bottom:=ActionImgPadding}
                                                        Widget:=texture_block{DefaultDesiredSize:=vector2{X:=XSizeActionUI,Y:=YSizeActionUI},DefaultTint:=CoalColor,DefaultImage:=Images.Img_LineBackground}
                                                    overlay_slot:
                                                        HorizontalAlignment:=horizontal_alignment.Fill
                                                        VerticalAlignment:=vertical_alignment.Fill
                                                        Widget:=texture_block{DefaultDesiredSize:=vector2{X:=XSizeActionUI,Y:=YSizeActionUI},DefaultImage:=Images.Img_LineBackground}
                                        overlay_slot:
                                            HorizontalAlignment:=horizontal_alignment.Left
                                            VerticalAlignment:=vertical_alignment.Center
                                            Padding:=margin{Left:=RowLeftPadding,Top:=RowTopPadding,Right:=RowRightPadding,Bottom:=RowBottomPadding}
                                            Widget:=stack_box:
                                                Orientation:=orientation.Horizontal
                                                    Slots:=array:
                                                        stack_box_slot:
                                                            Padding:=margin{Left:=InterColumnLeftPadding,Top:=InterColumnTopPadding,Right:=InterColumnRightPadding,Bottom:=InterColumnBottomPadding}
                                                            HorizontalAlignment:=horizontal_alignment.Left
                                                            VerticalAlignment:=vertical_alignment.Center
                                                            Widget:=CoalResource_Icon
                                                        stack_box_slot:
                                                            Padding:=margin{Left:=InterColumnLeftPadding,Top:=InterColumnTopPadding,Right:=InterColumnRightPadding,Bottom:=InterColumnBottomPadding}
                                                            HorizontalAlignment:=horizontal_alignment.Center
                                                            VerticalAlignment:=vertical_alignment.Center
                                                            Widget:=CoalResourceAmmount_TextBlock 
                #Mutant resource ammount
                canvas_slot :
                    Anchors:=anchors{Minimum:=vector2{X:=0.5,Y:=0.5},Maximum:=vector2{X:=0.5,Y:=0.5}}
                    ZOrder:=0
                    SizeToContent:=true
                    Offsets:=margin{Left:=XPositionUI_MutantRes,Top:=YPositionUI_MutantRes,Right:=0.0,Bottom:=0.0} #PositionX, PositionY, SizeX, SizeY
                    Alignment:=vector2{X:=0.5,Y:=0.5}
                    Widget:=overlay:
                        Slots:=array:
                            overlay_slot:
                                HorizontalAlignment:=horizontal_alignment.Center
                                VerticalAlignment:=vertical_alignment.Center
                                Widget:=overlay:
                                    Slots:=array:
                                        overlay_slot:
                                            Widget:=MutantResource_Slider
                            overlay_slot:
                                HorizontalAlignment:=horizontal_alignment.Center
                                VerticalAlignment:=vertical_alignment.Top
                                Widget:=overlay:
                                    Slots:=array:
                                        overlay_slot:
                                            HorizontalAlignment:=horizontal_alignment.Center
                                            VerticalAlignment:=vertical_alignment.Center
                                            Widget:=texture_block{DefaultDesiredSize:=vector2{X:=XSizeActionUI,Y:=YSizeActionUI},DefaultImage:=Images.EmptyTexture}
                                        overlay_slot:
                                            HorizontalAlignment:=horizontal_alignment.Fill
                                            VerticalAlignment:=vertical_alignment.Fill
                                            Widget:=overlay:
                                                Slots:=array:
                                                    overlay_slot:
                                                        HorizontalAlignment:=horizontal_alignment.Fill
                                                        VerticalAlignment:=vertical_alignment.Fill
                                                        Padding:=margin{Left:=ActionImgPadding,Top:=ActionImgPadding,Right:=ActionImgPadding,Bottom:=ActionImgPadding}
                                                        Widget:=texture_block{DefaultDesiredSize:=vector2{X:=XSizeActionUI,Y:=YSizeActionUI},DefaultTint:=MutantMaterialColor,DefaultImage:=Images.Img_LineBackground}
                                                    overlay_slot:
                                                        HorizontalAlignment:=horizontal_alignment.Fill
                                                        VerticalAlignment:=vertical_alignment.Fill
                                                        Widget:=texture_block{DefaultDesiredSize:=vector2{X:=XSizeActionUI,Y:=YSizeActionUI},DefaultImage:=Images.Img_LineBackground}
                                        overlay_slot:
                                            HorizontalAlignment:=horizontal_alignment.Left
                                            VerticalAlignment:=vertical_alignment.Center
                                            Padding:=margin{Left:=RowLeftPadding,Top:=RowTopPadding,Right:=RowRightPadding,Bottom:=RowBottomPadding}
                                            Widget:=stack_box:
                                                Orientation:=orientation.Horizontal
                                                    Slots:=array:
                                                        stack_box_slot:
                                                            Padding:=margin{Left:=InterColumnLeftPadding,Top:=InterColumnTopPadding,Right:=InterColumnRightPadding,Bottom:=InterColumnBottomPadding}
                                                            HorizontalAlignment:=horizontal_alignment.Left
                                                            VerticalAlignment:=vertical_alignment.Center
                                                            Widget:=MutantResource_Icon
                                                        stack_box_slot:
                                                            Padding:=margin{Left:=InterColumnLeftPadding,Top:=InterColumnTopPadding,Right:=InterColumnRightPadding,Bottom:=InterColumnBottomPadding}
                                                            HorizontalAlignment:=horizontal_alignment.Center
                                                            VerticalAlignment:=vertical_alignment.Center
                                                            Widget:=MutantResourceAmmount_TextBlock 
                                                                     
MakeActionInfoUI<constructor>(
    InBakeryMachinesManager : bakery_machines_manager)<transacts> := action_info_ui :
        BakeryMachinesManager := InBakeryMachinesManager
#MakeActionInfoUI<constructor>()<transacts> := action_info_ui :
                                                                                                               