using { /Verse.org/Simulation }
using { /Verse.org/Colors/NamedColors }
using { /UnrealEngine.com/Temporary/SpatialMath }
using { /UnrealEngine.com/Temporary/UI }
using { /Fortnite.com/UI }
using { /Fortnite.com/Devices }

#################NEW TO UPLOAD###############
RemoveAgentFromTreasureRegistrationMap(ExampleMap:[agent][]int, Agent:agent):[agent][]int=
    var NewMap:[agent][]int = map{}
    for (Key -> Value : ExampleMap, Key <> Agent):
        set NewMap = ConcatenateMaps(NewMap, map{Key => Value})
    return NewMap
#################NEW TO UPLOAD################

ToTruncatedString_Only1Decimal(Number:float):[]char=
    UseDecimals : int = 1
    var Str:[]char = ToString(Number)
    if:
        DotIndex := FindCharIndex['.', Str]
    then:
        if:
            StopIndex := if (UseDecimals > 0) then Min(DotIndex+UseDecimals+1,Str.Length) else DotIndex
            Value:=Str.Slice[0, StopIndex] 
            set Str = Value 
    Str

ToTruncatedStringAdaptable(Number:float, Decimals:int):[]char=
    var Str:[]char = ToString(Number)
    if:
        DotIndex := FindCharIndex['.', Str]
    then: 
        if:
            StopIndex := if (Decimals > 0) then Min(DotIndex+Decimals+1,Str.Length) else DotIndex
            Value:=Str.Slice[0, StopIndex] 
            set Str = Value 
    Str

FindSpaceIndexBeforeLineEnd(LineLimit:int, Str:[]char):int=
    SearchChar : char = ' '
    var FoundIndex:int = -1
    var Found:logic = false
    var Index : int = LineLimit
    loop:
        if(Char:=Str[Index]):
            if (Char = SearchChar):
                ##Print("Space found on Index {Index}")
                set FoundIndex = Index
                return FoundIndex
            else:
                ##Print("Space not found on Index {Index}")
        set Index = Index - 1
        if(Index<0):
            break
    return FoundIndex

FindSpaceIndexBetweenLines(StartIndex : int, LineLimit:int, Str:[]char):int=
    SearchChar : char = ' '
    var FoundIndex:int = -1
    var Found:logic = false
    var Index : int = LineLimit+StartIndex
    loop:
        if(Char:=Str[Index]):
            if (Char = SearchChar):
                ##Print("Space found on Index {Index}")
                set FoundIndex = Index
                return FoundIndex
            else:
                ##Print("Space not found on Index {Index}")
        set Index = Index - 1
        if(Index<StartIndex):
            break
    return FoundIndex

AdaptStringToLineV2(OriginalString:string, MaxCharactersPerLine : int):[]char=
    var Str:[]char = OriginalString
    var CharCounter : int = 0
    var PosibleLineStart : int = 0
    var CurrentPos : int = 0
    loop:
        if(Char:=Str[CurrentPos]):
            <#
            if(Char<>'\'):
                set CharCounter = CharCounter + 1
            #else:
            #    set CharCounter = -1#>
            set CharCounter = CharCounter + 1
            set CurrentPos = CurrentPos + 1
            if(CharCounter > MaxCharactersPerLine):
                ##Print("Detected exceded line lenght on {CurrentPos}. Trying to make a jump new line Starting from {PosibleLineStart}")
                IndexForJumpLine:=FindSpaceIndexBetweenLines(PosibleLineStart,MaxCharactersPerLine,Str)
                ##Print("Adding jump line on {IndexForJumpLine}")
                set Str = AddJumpLineOnPosition(Str,IndexForJumpLine)
                ##Print("Final string adapted to: \n ########## \n {Str}")
                set CharCounter = -1
                set PosibleLineStart = IndexForJumpLine+1
                #set CurrentPos = PosibleLineStart          
        else:
            return Str 


AddJumpLineOnPosition(OriginalCharArray : []char, IndexForJumpLine : int) : []char = 
    var OutPut : []char = ""
    if(IndexForJumpLine<>-1):
        for(Index:=0..IndexForJumpLine-1,Character:=OriginalCharArray[Index]):
            set OutPut = OutPut + array{Character}
        set OutPut = OutPut + "\n"
        for(Index:=IndexForJumpLine+1..OriginalCharArray.Length-1, Character:=OriginalCharArray[Index]):
            set OutPut = OutPut + array{Character}
        return OutPut
    else:
        return OriginalCharArray


AdaptStringToLine(MessageString:string, MaxCharactersPerLine : int):[]char=
    var Str:[]char = MessageString
    ##Print("Str.Length is {Str.Length}")
    if(Str.Length > MaxCharactersPerLine):
        IndexForJumpLine:=FindSpaceIndexBeforeLineEnd(MaxCharactersPerLine,MessageString)
        if(IndexForJumpLine<>-1):
            ##Print("JumpLine on {IndexForJumpLine}")
            var FinalString : []char = ""
            for(Index:=0..IndexForJumpLine-1,Character:=Str[Index]):
                set FinalString = FinalString + array{Character}
            set FinalString = FinalString + "\n"
            for(Index:=IndexForJumpLine+1..Str.Length-1, Character:=Str[Index]):
                set FinalString = FinalString + array{Character}
            return FinalString
    return Str

#It does not work!
RoundToDecimalPlaces(Value:float, DecimalPlaces:int)<decides>:float = #Truncates float decimals
    DecimalPlaces >= 0
    Multiplier := Pow(10.0, DecimalPlaces * 1.0)
    RoundedValues := Round[Value * Multiplier] * 1.0
    return RoundedValues / Multiplier


GetForwardRotation_Vector2(Destiny : vector3, Origin : vector3) : rotation =
    #Check this for future https://www.youtube.com/watch?v=A3OuFLHum6w
    Direction := Destiny - Origin
    X:=Direction.X
    Y:=Direction.Y

    var Angle : float = RadiansToDegrees(ArcTan(Y/X))

    if(X<0.0):
        set Angle = Angle + 180.0

    return MakeRotationFromYawPitchRollDegrees(Angle, 0.0, 0.0)

GetForwardRotation_Vector3(Destiny : vector3, Origin : vector3) : rotation =
    #Check this for future https://www.youtube.com/watch?v=A3OuFLHum6w
    #Roll is X: Red Axis OF THE OBJECT
    #Pitch is Y: Green Axis OF THE OBJECT
    #Yaw is Z: Blue axis OF THE OBJECT
    Direction := Destiny - Origin
    X:=Direction.X
    Y:=Direction.Y

    DistanceBetweenXY:=DistanceXY(Destiny,Origin)
    DistanceBetweenh:=Destiny.Z-Origin.Z

    var AngleX : float = RadiansToDegrees(ArcTan(Y/X))
    var AngleY : float = RadiansToDegrees(ArcTan(DistanceBetweenh/DistanceBetweenXY)) #VerticalAngle

    if(X<0.0):
        set AngleX = AngleX + 180.0
    if(Y<0.0):
        #set AngleY = AngleY + 180.0

    ##Print("AngleX: {GetOptimizedMagnitudeLetter(AngleX)} |AngleY: {GetOptimizedMagnitudeLetter(AngleY)}")

    return MakeRotationFromYawPitchRollDegrees(AngleX, AngleY,0.0)

SetStar(Prop : creative_prop, Tier : int) : void =
    if(MaterialStar:=AllStarsMaterials[Tier-1]):
        Prop.SetMaterial(MaterialStar,?Index:=0)
        
AdaptTimeDigit(Number:int):[]char=
    var Str:[]char = ToString(Number)
    if(Str.Length=1):
        set Str = "0"+Str
        return Str
    return Str

TransformToMinutes_Seconds_FromFloat(TotalSecondsFloat : float) : tuple([]char,[]char) =
    var Minutes : []char = "00"
    var Seconds : []char = "00"
    if(TotalSecondsInt:=Ceil[TotalSecondsFloat]):
        return TransformToMinutes_Seconds(TotalSecondsInt)
    return(Minutes,Seconds)


GetDoubleDecimalMagnitudeLetter(FloatAmmount : float, PreCharacter : string, PostCharacter : string) : string =
    var FinalString : string = ""
    if(FloatAmmount < Pow(10.0,30.0)): #QUETTA
        if(FloatAmmount < Pow(10.0,27.0)): #RONNA
            if(FloatAmmount < Pow(10.0,24.0)): #Yotta
                if(FloatAmmount < Pow(10.0,21.0)): #Zetta
                    if(FloatAmmount < Pow(10.0,18.0)): #Exa
                        if(FloatAmmount < Pow(10.0,15.0)): #Peta
                            if(FloatAmmount < Pow(10.0,12.0)): #Trillions
                                if(FloatAmmount < Pow(10.0,9.0)): #Billions
                                    if(FloatAmmount < Pow(10.0,6.0)): #Millions
                                        if(FloatAmmount < Pow(10.0,3.0)): #Kilos
                                            set FinalString = PreCharacter + ToTruncatedStringAdaptable(FloatAmmount,2) + PostCharacter
                                        else:
                                                Value:=(FloatAmmount * 1.0) / Pow(10.0,3.0)
                                                String := ToTruncatedStringAdaptable(Value,2)
                                                set FinalString = PreCharacter + String + " k"
                                    else:
                                        Value:=(FloatAmmount * 1.0) / Pow(10.0,6.0)
                                        String := ToTruncatedString(Value,2)
                                        set FinalString = PreCharacter + String + " M"
                                else:
                                    Value:=(FloatAmmount * 1.0) / Pow(10.0,9.0)
                                    String := ToTruncatedStringAdaptable(Value,2)
                                    set FinalString = PreCharacter + String + " G"
                            else:
                                Value:=(FloatAmmount * 1.0) / Pow(10.0,12.0)
                                String := ToTruncatedStringAdaptable(Value,2)
                                set FinalString = PreCharacter + String + " T"
                        else:
                            Value:=(FloatAmmount * 1.0) / Pow(10.0,15.0)
                            String := ToTruncatedStringAdaptable(Value,2)
                            set FinalString = PreCharacter + String + " P"
                    else:
                        Value:=(FloatAmmount * 1.0) / Pow(10.0,18.0)
                        String := ToTruncatedStringAdaptable(Value,2)
                        set FinalString = PreCharacter + String + " E"
                else:
                    Value:=(FloatAmmount * 1.0) / Pow(10.0,21.0)
                    String := ToTruncatedStringAdaptable(Value,2)
                    set FinalString = PreCharacter + String + " Z"
            else:
                Value:=(FloatAmmount * 1.0) / Pow(10.0,24.0)
                String := ToTruncatedStringAdaptable(Value,2)
                set FinalString = PreCharacter + String + " Y"
        else:
            Value:=(FloatAmmount * 1.0) / Pow(10.0,27.0)
            String := ToTruncatedStringAdaptable(Value,2)
            set FinalString = PreCharacter + String + " R"
    else:
        Value:=(FloatAmmount * 1.0) / Pow(10.0,30.0)
        String := ToTruncatedStringAdaptable(Value,2)
        set FinalString = PreCharacter + String + " Q"
    return FinalString
    
GetTripleDecimalMagnitudeLetter(FloatAmmount : float, PreCharacter : string, PostCharacter : string) : string =
    var FinalString : string = ""
    if(FloatAmmount < Pow(10.0,30.0)): #QUETTA
        if(FloatAmmount < Pow(10.0,27.0)): #RONNA
            if(FloatAmmount < Pow(10.0,24.0)): #Yotta
                if(FloatAmmount < Pow(10.0,21.0)): #Zetta
                    if(FloatAmmount < Pow(10.0,18.0)): #Exa
                        if(FloatAmmount < Pow(10.0,15.0)): #Peta
                            if(FloatAmmount < Pow(10.0,12.0)): #Trillions
                                if(FloatAmmount < Pow(10.0,9.0)): #Billions
                                    if(FloatAmmount < Pow(10.0,6.0)): #Millions
                                        if(FloatAmmount < Pow(10.0,3.0)): #Kilos
                                            set FinalString = PreCharacter + ToTruncatedStringAdaptable(FloatAmmount,2) + PostCharacter
                                        else:
                                                Value:=(FloatAmmount * 1.0) / Pow(10.0,3.0)
                                                String := ToTruncatedStringAdaptable(Value,3)
                                                set FinalString = PreCharacter + String + " k"
                                    else:
                                        Value:=(FloatAmmount * 1.0) / Pow(10.0,6.0)
                                        String := ToTruncatedString(Value,3)
                                        set FinalString = PreCharacter + String + " M"
                                else:
                                    Value:=(FloatAmmount * 1.0) / Pow(10.0,9.0)
                                    String := ToTruncatedStringAdaptable(Value,3)
                                    set FinalString = PreCharacter + String + " G"
                            else:
                                Value:=(FloatAmmount * 1.0) / Pow(10.0,12.0)
                                String := ToTruncatedStringAdaptable(Value,3)
                                set FinalString = PreCharacter + String + " T"
                        else:
                            Value:=(FloatAmmount * 1.0) / Pow(10.0,15.0)
                            String := ToTruncatedStringAdaptable(Value,3)
                            set FinalString = PreCharacter + String + " P"
                    else:
                        Value:=(FloatAmmount * 1.0) / Pow(10.0,18.0)
                        String := ToTruncatedStringAdaptable(Value,3)
                        set FinalString = PreCharacter + String + " E"
                else:
                    Value:=(FloatAmmount * 1.0) / Pow(10.0,21.0)
                    String := ToTruncatedStringAdaptable(Value,3)
                    set FinalString = PreCharacter + String + " Z"
            else:
                Value:=(FloatAmmount * 1.0) / Pow(10.0,24.0)
                String := ToTruncatedStringAdaptable(Value,3)
                set FinalString = PreCharacter + String + " Y"
        else:
            Value:=(FloatAmmount * 1.0) / Pow(10.0,27.0)
            String := ToTruncatedStringAdaptable(Value,3)
            set FinalString = PreCharacter + String + " R"
    else:
        Value:=(FloatAmmount * 1.0) / Pow(10.0,30.0)
        String := ToTruncatedStringAdaptable(Value,3)
        set FinalString = PreCharacter + String + " Q"
    return FinalString
    
#PET SETTINGS
SetPetHealthBar(Pet : hexagame_pet, CurrentHealth : float, TextureBlock : texture_block, TextBlock : text_block) : void =
    Percentage := CurrentHealth / Pet.MaxHealth
    TextureBlock.SetDesiredSize(vector2{X:=MaxHealthSize_PetHexaUI*Percentage,Y:=10.0})
    TextBlock.SetText(BasicMessage("❤️ {GetDoubleDecimalMagnitudeLetter(CurrentHealth,"","")}"))
    TextBlock.SetTextColor(Lime)

    #if(Pet.Health>0.0):
    #    TextBlock.SetTextColor(White)
    #else:
        

SetPetImage(Pet : hexagame_pet, TextureBlock : texture_block) : void =
    if(Pet.PetType=ADC):
        TextureBlock.SetImage(Images.Icon_ADC)
    else if(Pet.PetType=DRUID):
        TextureBlock.SetImage(Images.Icon_Druid)
    else if(Pet.PetType=TANK):
        TextureBlock.SetImage(Images.Icon_Tank)
    else if(Pet.PetType=FIGHTER):
        TextureBlock.SetImage(Images.Icon_Fighter)
    else if(Pet.PetType=HUNTER):
        TextureBlock.SetImage(Images.Icon_Hunter)
    else:
        TextureBlock.SetImage(Images.Cross_Icon)
    if(Pet.Health>0.0):
        TextureBlock.SetTint(White)
    else:
        TextureBlock.SetTint(Red)

        <#
SetPetClass(Pet : hexagame_pet, Class_TextBlock : text_block) : void =
    if(Pet.PetType=ADC):
        Class_TextBlock.SetText(BasicMessage("ADC [{Pet.PackA_LV}][{Pet.PackB_LV}]"))
    else if(Pet.PetType=DRUID):
        Class_TextBlock.SetText(BasicMessage("Druid [{Pet.PackA_LV}][{Pet.PackB_LV}]"))
    else if(Pet.PetType=TANK):
        Class_TextBlock.SetText(BasicMessage("Tank [{Pet.PackA_LV}][{Pet.PackB_LV}]"))
    else if(Pet.PetType=FIGHTER):
        Class_TextBlock.SetText(BasicMessage("Fighter [{Pet.PackA_LV}][{Pet.PackB_LV}]"))
    else if(Pet.PetType=HUNTER):
        Class_TextBlock.SetText(BasicMessage("Hunter [{Pet.PackA_LV}][{Pet.PackB_LV}]"))
        #>

SetPetClass(Pet : hexagame_pet, Class_TextBlock : text_block, LVPackA_TextBlock : text_block, LVPackB_TextBlock : text_block) : void =
    if: 
        ColorA_ToUse := AllTierColors[Pet.PackA_LV-1]
        ColorB_ToUse := AllTierColors[Pet.PackB_LV-1]
    then:
        LVPackA_TextBlock.SetTextColor(ColorA_ToUse)
        LVPackA_TextBlock.SetText(BasicMessage("[{Pet.PackA_LV}]"))
        LVPackB_TextBlock.SetTextColor(ColorB_ToUse)
        LVPackB_TextBlock.SetText(BasicMessage("[{Pet.PackB_LV}]"))

    if(Pet.PetType=ADC):
        Class_TextBlock.SetText(BasicMessage("ADC"))
    else if(Pet.PetType=DRUID):
        Class_TextBlock.SetText(BasicMessage("Druid"))
    else if(Pet.PetType=TANK):
        Class_TextBlock.SetText(BasicMessage("Tank"))
    else if(Pet.PetType=FIGHTER):
        Class_TextBlock.SetText(BasicMessage("Fighter"))
    else if(Pet.PetType=HUNTER):
        Class_TextBlock.SetText(BasicMessage("Hunter"))

SetPetPurity(Pet : hexagame_pet, Tier : int, TextBlock : text_block) : void =
    var FinalString : string = "✨ {Pet.Purity}%"
    if(Tier=1):
        set FinalString = FinalString + " ⭐"
    else if(Tier=2):
        set FinalString = FinalString + " ⭐⭐"
    else if(Tier=3):
        set FinalString = FinalString + " ⭐⭐⭐"
    TextBlock.SetText(BasicMessage(FinalString))

SetPetHealthBar(MaxHealth : float, CurrentHealth : float, TextureBlock : texture_block, TextBlock : text_block) : void =
    Percentage := CurrentHealth / MaxHealth
    TextureBlock.SetDesiredSize(vector2{X:=MaxHealthSize_PetHexaUI*Percentage,Y:=10.0})
    TextBlock.SetText(BasicMessage("❤️ {GetOptimizedMagnitudeLetter(CurrentHealth)}"))

SetPetImage(Class : int, TextureBlock : texture_block) : void =
    if(Class=ADC):
        TextureBlock.SetImage(Images.Icon_ADC)
    else if(Class=DRUID):
        TextureBlock.SetImage(Images.Icon_Druid)
    else if(Class=TANK):
        TextureBlock.SetImage(Images.Icon_Tank)
    else if(Class=FIGHTER):
        TextureBlock.SetImage(Images.Icon_Fighter)
    else if(Class=HUNTER):
        TextureBlock.SetImage(Images.Icon_Hunter)
    else:
        TextureBlock.SetImage(Images.Cross_Icon)

SetPetClass(Pet : pet_data_table, Class_TextBlock : text_block, LVPackA_TextBlock : text_block, LVPackB_TextBlock : text_block) : void =
    if: 
        ColorA_ToUse := AllTierColors[Pet.LVPet_PackageA-1]
        ColorB_ToUse := AllTierColors[Pet.LVPet_PackageB-1]
    then:
        LVPackA_TextBlock.SetTextColor(ColorA_ToUse)
        LVPackA_TextBlock.SetText(BasicMessage("[{Pet.LVPet_PackageA}]"))
        LVPackB_TextBlock.SetTextColor(ColorB_ToUse)
        LVPackB_TextBlock.SetText(BasicMessage("[{Pet.LVPet_PackageB}]"))

    if(Pet.Type=ADC):
        Class_TextBlock.SetText(BasicMessage("ADC"))
    else if(Pet.Type=DRUID):
        Class_TextBlock.SetText(BasicMessage("Druid"))
    else if(Pet.Type=TANK):
        Class_TextBlock.SetText(BasicMessage("Tank"))
    else if(Pet.Type=FIGHTER):
        Class_TextBlock.SetText(BasicMessage("Fighter"))
    else if(Pet.Type=HUNTER):
        Class_TextBlock.SetText(BasicMessage("Hunter"))

        <#
SetPetClass(Pet : pet_data_table, TextBlock : text_block) : void =
    if(Pet.Type=ADC):
        TextBlock.SetText(BasicMessage("ADC [{Pet.LVPet_PackageA}][{Pet.LVPet_PackageB}]"))
    else if(Pet.Type=DRUID):
        TextBlock.SetText(BasicMessage("Druid [{Pet.LVPet_PackageA}][{Pet.LVPet_PackageB}]"))
    else if(Pet.Type=TANK):
        TextBlock.SetText(BasicMessage("Tank [{Pet.LVPet_PackageA}][{Pet.LVPet_PackageB}]"))
    else if(Pet.Type=FIGHTER):
        TextBlock.SetText(BasicMessage("Fighter [{Pet.LVPet_PackageA}][{Pet.LVPet_PackageB}]"))
    else if(Pet.Type=HUNTER):
        TextBlock.SetText(BasicMessage("Hunter [{Pet.LVPet_PackageA}][{Pet.LVPet_PackageB}]"))#>
        
SetPetPurity(Purity : int, TextBlock : text_block) : void =
    var FinalString : string = "✨ {Purity}%"
    TextBlock.SetText(BasicMessage(FinalString))

SetPetTier(Tier : int, TextBlock : text_block) : void =
    var FinalString : string = " "
    if(Tier=1):
        set FinalString = "⭐"
    else if(Tier=2):
        set FinalString = "⭐⭐"
    else if(Tier=3):
        set FinalString = FinalString + "⭐⭐⭐"
    TextBlock.SetText(BasicMessage(FinalString))

SetPetPurityAndTier(Purity : int, Tier : int, TextBlock : text_block) : void =
    var FinalString : string = "✨ {Purity}% Purity"
    if(Tier=1):
        set FinalString = FinalString + " ⭐"
    else if(Tier=2):
        set FinalString = FinalString + " ⭐⭐"
    else if(Tier=3):
        set FinalString = FinalString + " ⭐⭐⭐"
    TextBlock.SetText(BasicMessage(FinalString))

SetHealthUI(Prop : creative_prop, MaxHealth : float, Health : float, IsPlayerPet : logic) : void =
    for(HealthBarFragment:=3..12):
        if(IsPlayerPet?):
            Prop.SetMaterial(DarkGreenPetBarMaterial,?Index:=HealthBarFragment)
        else:
            Prop.SetMaterial(DarkRedPetBarMaterial,?Index:=HealthBarFragment)
    if(CurrentHealthBarFragment:=Ceil[(Health/MaxHealth)*10.0]):
        if(Health>0.0):
            for(HealthBarFragment:=2..CurrentHealthBarFragment+2):
                if(IsPlayerPet?):
                    Prop.SetMaterial(FriendlyPetBarMaterial,?Index:=HealthBarFragment)
                else:
                    Prop.SetMaterial(EnemyPetBarMaterial,?Index:=HealthBarFragment)
  
RemoveHealthUI(Prop : creative_prop) : void =
    for(HealthBarFragment:=3..12):
        Prop.SetMaterial(Materials.MI_Life_0,?Index:=HealthBarFragment)
        
                      