using { /UnrealEngine.com/Temporary/UI }
using { /UnrealEngine.com/Temporary/SpatialMath }
using { /Verse.org/Simulation }
using { /Fortnite.com/Devices }
using { /Fortnite.com/UI }
using { /Verse.org/Colors/NamedColors }
using { /Verse.org/Colors }

# See https://dev.epicgames.com/documentation/en-us/uefn/create-your-own-device-in-verse for how to create a verse device.

# A Verse-authored creative device that can be placed in a level
backpack_upgrades_ui := class():

    #Class settings
    var MaybeAgent : ?agent = false

    BakeryMachinesManager : bakery_machines_manager
    
    BackgroundImageSizeX : float = 1375.0
    BackgroundImageSizeY : float = 1100.0

    #LineSizeX : float = 1000.0
    #LineSizeY : float = 80.0 #Size of button regular

    PTopColumn1SizeX : float = 350.0
    PTopColumn1SizeY : float = 80.0
    PTopColumn2SizeX : float = 135.0
    PTopColumn2SizeY : float = 80.0

    BUColumn1SizeX : float = 80.0
    BUColumn1SizeY : float = 80.0
    BUColumn2SizeX : float = 425.0
    BUColumn2SizeY : float = 80.0
    BUColumn3SizeX : float = 150.0
    BUColumn3SizeY : float = 80.0
    BUColumn4SizeX : float = 200.0
    BUColumn4SizeY : float = 80.0
    BUColumn5SizeX : float = 200.0
    BUColumn5SizeY : float = 80.0
    BUColumn6SizeX : float = 225.0
    BUColumn6SizeY : float = 80.0

    TitleTextBlock : text_block = text_block{DefaultTextColor:=White, DefaultShadowOffset:=ShadowOffset, DefaultShadowOpacity := DefaultTextShadowOpacity}

    Sapphires_Texture : texture_block = texture_block{DefaultDesiredSize:=vector2{X:=IconSmallSizeX,Y:=IconSmallSizeY}, DefaultImage:=Images.Sapphire_Icon}
    TotalSapphires_BackgroundTexture : texture_block = texture_block{DefaultDesiredSize:=vector2{X:=SizeBoxSaphiresX,Y:=SizeBoxSaphiresY}, DefaultImage:=Images.Img_LineBackground2}
    
    TotalWater_BackgroundTexture : texture_block = texture_block{DefaultDesiredSize:=vector2{X:=SizeBoxSaphiresX,Y:=SizeBoxSaphiresY}, DefaultImage:=Images.Img_LineBackground2}
    TotalFlour_BackgroundTexture : texture_block = texture_block{DefaultDesiredSize:=vector2{X:=SizeBoxSaphiresX,Y:=SizeBoxSaphiresY}, DefaultImage:=Images.Img_LineBackground2}
    TotalCheese_BackgroundTexture : texture_block = texture_block{DefaultDesiredSize:=vector2{X:=SizeBoxSaphiresX,Y:=SizeBoxSaphiresY}, DefaultImage:=Images.Img_LineBackground2}
    TotalPimentos_BackgroundTexture : texture_block = texture_block{DefaultDesiredSize:=vector2{X:=SizeBoxSaphiresX,Y:=SizeBoxSaphiresY}, DefaultImage:=Images.Img_LineBackground2}
    #TotalMutantMaterial_BackgroundTexture : texture_block = texture_block{DefaultDesiredSize:=vector2{X:=SizeBoxSaphiresX,Y:=SizeBoxSaphiresY}, DefaultImage:=Images.Img_LineBackground2}

    TotalWater_TextBlock : text_block = text_block{DefaultTextColor:=WaterTextColor,DefaultShadowOffset:=ShadowOffset, DefaultShadowOpacity := DefaultTextShadowOpacity}
    TotalFlour_TextBlock : text_block = text_block{DefaultTextColor:=FlourTextColor,DefaultShadowOffset:=ShadowOffset, DefaultShadowOpacity := DefaultTextShadowOpacity}
    TotalCheese_TextBlock : text_block = text_block{DefaultTextColor:=CheeseTextColor,DefaultShadowOffset:=ShadowOffset, DefaultShadowOpacity := DefaultTextShadowOpacity}
    TotalPimentos_TextBlock : text_block = text_block{DefaultTextColor:=PimentoTextColor,DefaultShadowOffset:=ShadowOffset, DefaultShadowOpacity := DefaultTextShadowOpacity}
    TotalMutantMaterial_TextBlock : text_block = text_block{DefaultTextColor:=MutantMaterialTextColor,DefaultShadowOffset:=ShadowOffset, DefaultShadowOpacity := DefaultTextShadowOpacity}

    BU_Row1_Icon : texture_block = texture_block{DefaultDesiredSize:=vector2{X:=IconSizeX,Y:=IconSizeY},DefaultImage:=Images.Water_Icon}
    BU_Row2_Icon : texture_block = texture_block{DefaultDesiredSize:=vector2{X:=IconSizeX,Y:=IconSizeY},DefaultImage:=Images.Flour_Icon}
    BU_Row3_Icon : texture_block = texture_block{DefaultDesiredSize:=vector2{X:=IconSizeX,Y:=IconSizeY},DefaultImage:=Images.Cheese_Icon}
    BU_Row4_Icon : texture_block = texture_block{DefaultDesiredSize:=vector2{X:=IconSizeX,Y:=IconSizeY},DefaultImage:=Images.Pimento_Icon}
    BU_Row5_Icon : texture_block = texture_block{DefaultDesiredSize:=vector2{X:=IconSizeX,Y:=IconSizeY},DefaultImage:=Images.Mutant_Material_Icon}

    BU_Row1_Current_TextBlock : text_block = text_block{DefaultTextColor:=CurrentTextColor, DefaultShadowOffset:=ShadowOffset, DefaultShadowOpacity := DefaultTextShadowOpacity}
    BU_Row2_Current_TextBlock : text_block = text_block{DefaultTextColor:=CurrentTextColor, DefaultShadowOffset:=ShadowOffset, DefaultShadowOpacity := DefaultTextShadowOpacity}
    BU_Row3_Current_TextBlock : text_block = text_block{DefaultTextColor:=CurrentTextColor, DefaultShadowOffset:=ShadowOffset, DefaultShadowOpacity := DefaultTextShadowOpacity}
    BU_Row4_Current_TextBlock : text_block = text_block{DefaultTextColor:=CurrentTextColor, DefaultShadowOffset:=ShadowOffset, DefaultShadowOpacity := DefaultTextShadowOpacity}
    BU_Row5_Current_TextBlock : text_block = text_block{DefaultTextColor:=CurrentTextColor, DefaultShadowOffset:=ShadowOffset, DefaultShadowOpacity := DefaultTextShadowOpacity}

    BU_Row1_Upgrade_TextBlock : text_block = text_block{DefaultTextColor:=UpgradeTextColor, DefaultShadowOffset:=ShadowOffset, DefaultShadowOpacity := DefaultTextShadowOpacity}
    BU_Row2_Upgrade_TextBlock : text_block = text_block{DefaultTextColor:=UpgradeTextColor, DefaultShadowOffset:=ShadowOffset, DefaultShadowOpacity := DefaultTextShadowOpacity}
    BU_Row3_Upgrade_TextBlock : text_block = text_block{DefaultTextColor:=UpgradeTextColor, DefaultShadowOffset:=ShadowOffset, DefaultShadowOpacity := DefaultTextShadowOpacity}
    BU_Row4_Upgrade_TextBlock : text_block = text_block{DefaultTextColor:=UpgradeTextColor, DefaultShadowOffset:=ShadowOffset, DefaultShadowOpacity := DefaultTextShadowOpacity}
    BU_Row5_Upgrade_TextBlock : text_block = text_block{DefaultTextColor:=UpgradeTextColor, DefaultShadowOffset:=ShadowOffset, DefaultShadowOpacity := DefaultTextShadowOpacity}

    BU_Row1_UpgradeCost_TextBlock : text_block = text_block{DefaultTextColor:=SapphiresTextColor, DefaultShadowOffset:=ShadowOffset, DefaultShadowOpacity := DefaultTextShadowOpacity}
    BU_Row2_UpgradeCost_TextBlock : text_block = text_block{DefaultTextColor:=SapphiresTextColor, DefaultShadowOffset:=ShadowOffset, DefaultShadowOpacity := DefaultTextShadowOpacity}
    BU_Row3_UpgradeCost_TextBlock : text_block = text_block{DefaultTextColor:=SapphiresTextColor, DefaultShadowOffset:=ShadowOffset, DefaultShadowOpacity := DefaultTextShadowOpacity}
    BU_Row4_UpgradeCost_TextBlock : text_block = text_block{DefaultTextColor:=SapphiresTextColor, DefaultShadowOffset:=ShadowOffset, DefaultShadowOpacity := DefaultTextShadowOpacity}
    BU_Row5_UpgradeCost_TextBlock : text_block = text_block{DefaultTextColor:=SapphiresTextColor, DefaultShadowOffset:=ShadowOffset, DefaultShadowOpacity := DefaultTextShadowOpacity}

    BU_Row1Button : button_quiet = button_quiet{}
    BU_Row2Button : button_quiet = button_quiet{}
    BU_Row3Button : button_quiet = button_quiet{}
    BU_Row4Button : button_quiet = button_quiet{}
    BU_Row5Button : button_quiet = button_quiet{}

    var MaybeCanvas : ?canvas = false
    var UIIsDisplayed : logic = false

    var TimeElapsed : int = 0

    #Close
    CloseButton : button_loud = button_loud{}

    #Arrays
    var AllResourcersColor : []color = array{}

    var AllSFP_TopRows_TextBlock : []text_block = array{}
    var AllSFP_Resourcers_TextBlock : []text_block = array{}
    var AllSFP_Rows_IconTexture : []texture_block = array{}

    var AllBU_Rows_Current_TextBlock : []text_block = array{}
    var AllBU_Rows_Upgrade_TextBlock : []text_block = array{}
    var AllBU_Rows_UpgradeCost_TextBlock : []text_block = array{}
    var AllBU_RowButtons : []button_quiet = array{}

    InitUI() : void =
        SetArrays()
        InitSubscribers()

        Canvas := MakeCanvas()
        set MaybeCanvas = option{Canvas}

    SetArrays() : void =

        set AllResourcersColor = array{WaterTextColor, FlourTextColor, CheeseTextColor, PimentoTextColor, MutantMaterialTextColor}
        set AllSFP_Resourcers_TextBlock = AllSFP_Resourcers_TextBlock
        
        set AllBU_Rows_Current_TextBlock = array{BU_Row1_Current_TextBlock,BU_Row2_Current_TextBlock,BU_Row3_Current_TextBlock,BU_Row4_Current_TextBlock, BU_Row5_Current_TextBlock}
        set AllBU_Rows_Upgrade_TextBlock = array{BU_Row1_Upgrade_TextBlock,BU_Row2_Upgrade_TextBlock,BU_Row3_Upgrade_TextBlock,BU_Row4_Upgrade_TextBlock, BU_Row5_Upgrade_TextBlock}
        set AllBU_Rows_UpgradeCost_TextBlock = array{BU_Row1_UpgradeCost_TextBlock,BU_Row2_UpgradeCost_TextBlock,BU_Row3_UpgradeCost_TextBlock,BU_Row4_UpgradeCost_TextBlock, BU_Row5_UpgradeCost_TextBlock}

        set AllBU_RowButtons = array{BU_Row1Button,BU_Row2Button,BU_Row3Button,BU_Row4Button} 

    InitSubscribers() : void =

        #Drone speed
        BU_Row1Button.OnClick().Subscribe(handler_upgrades{MaybeBakeryMachinesManager := option{BakeryMachinesManager}, Resource := WATER}.HandlerBackpackUpgrade)  
        BU_Row2Button.OnClick().Subscribe(handler_upgrades{MaybeBakeryMachinesManager := option{BakeryMachinesManager}, Resource := FLOUR}.HandlerBackpackUpgrade)
        BU_Row3Button.OnClick().Subscribe(handler_upgrades{MaybeBakeryMachinesManager := option{BakeryMachinesManager}, Resource := CHEESE}.HandlerBackpackUpgrade) 
        BU_Row4Button.OnClick().Subscribe(handler_upgrades{MaybeBakeryMachinesManager := option{BakeryMachinesManager}, Resource := PIMENTOS}.HandlerBackpackUpgrade)
        #BU_Row5Button.OnClick().Subscribe(handler_upgrades{MaybeBakeryMachinesManager := option{BakeryMachinesManager}, Resource := MUTANTMATERIAL}.HandlerBackpackUpgrade)    
 
        CloseButton.OnClick().Subscribe(Remove_Canvas) 

    SetAgent(InAgent : agent) : void =
        set MaybeAgent = option{InAgent}

    UpdateUI(PlayerData : player_data_table) : void =

        CloseButton.SetText(BasicMessage("CLOSE"))

        #Inventory resourcers
        TotalWaterAvailable := PlayerData.Inventory_Water + PlayerData.MainWaterDeposit_Value + PlayerData.WaterPreparedFor_B + PlayerData.WaterPreparedFor_CB + PlayerData.WaterPreparedFor_PCS
        TotalFlourAvailable := PlayerData.Inventory_Flour + PlayerData.MainFlourDeposit_Value + PlayerData.FlourPreparedFor_B + PlayerData.FlourPreparedFor_CB + PlayerData.FlourPreparedFor_PCS
        TotalCheeseAvailable := PlayerData.Inventory_Cheese + PlayerData.MainCheeseDeposit_Value + PlayerData.CheesePreparedFor_CB + PlayerData.CheesePreparedFor_PCS
        TotalPimentosAvailable := PlayerData.Inventory_Pimentos + PlayerData.MainPimentosDeposit_Value + PlayerData.PimentosPreparedFor_PCS

        TotalWater_TextBlock.SetText(BasicMessage("{GetOptimizedMagnitudeLetter(TotalWaterAvailable)}"))
        TotalFlour_TextBlock.SetText(BasicMessage("{GetOptimizedMagnitudeLetter(TotalFlourAvailable)}"))
        TotalCheese_TextBlock.SetText(BasicMessage("{GetOptimizedMagnitudeLetter(TotalCheeseAvailable)}"))
        TotalPimentos_TextBlock.SetText(BasicMessage("{GetOptimizedMagnitudeLetter(TotalPimentosAvailable)}"))
        TotalMutantMaterial_TextBlock.SetText(BasicMessage("{GetOptimizedMagnitudeLetter(PlayerData.Inventory_MutantMaterial)}"))

        for(Resource:=WATER..MUTANTMATERIAL):
            if:
                ResourceLevel := PlayerData.AllLV_BackpackLimits[Resource]
                BU_Row_Current_TextBlock := AllBU_Rows_Current_TextBlock[Resource]
                BU_Row_Upgrade_TextBlock := AllBU_Rows_Upgrade_TextBlock[Resource]
                BU_Row_UpgradeCost_TextBlock := AllBU_Rows_UpgradeCost_TextBlock[Resource]
                Color := AllResourcersColor[Resource]
            then:
                TupleBackpackLimits := CalculeBackpackUpgrade(ResourceLevel, Resource)

                BU_Row_Current_TextBlock.SetText(BasicMessage(GetMagnitudeLetter(TupleBackpackLimits(0),"","")))
                BU_Row_UpgradeCost_TextBlock.SetText(BasicMessage(GetMagnitudeLetter(TupleBackpackLimits(1),"x ","")))
                BU_Row_Upgrade_TextBlock.SetText(BasicMessage(GetMagnitudeLetter(TupleBackpackLimits(2),"+","")))

                BU_Row_Current_TextBlock.SetTextColor(CurrentTextColor)
                BU_Row_Upgrade_TextBlock.SetTextColor(UpgradeTextColor)

                TupleHasEnough:=BakeryMachinesManager.PlayersDataManager.HasEnoughResource(PlayerData,TupleBackpackLimits(1),Resource)
                if(TupleHasEnough(0)>-1):
                    BU_Row_UpgradeCost_TextBlock.SetTextColor(Color)
                else:
                    BU_Row_UpgradeCost_TextBlock.SetTextColor(NotEnoughTextColor)                

    Remove_Canvas(Message : widget_message) : void =
        set UIIsDisplayed = false
        if:
            Agent:=MaybeAgent?
            Player:=player[Agent]
            PlayerUI:=GetPlayerUI[Player]
            Canvas := MaybeCanvas?
        then:
            PlayerUI.RemoveWidget(Canvas)
            RemoveDelay_OnRemove()

    Remove_CanvasWithoutMessage() : void =
        set UIIsDisplayed = false
        ##Print("Trying to remove UI")
        if:
            Agent:=MaybeAgent?
            Player:=player[Agent]
            PlayerUI:=GetPlayerUI[Player]
            Canvas := MaybeCanvas?
        then:
            ##Print("Removing UI")
            PlayerUI.RemoveWidget(Canvas)
            RemoveDelay_OnRemove()

    RemoveDelay_OnAdd() : void =
        TitleTextBlock.SetText(BasicMessage("Backpack upgrades"))
        #CurrentBuyOption_textblock.SetVisibility(widget_visibility.Hidden)

    RemoveDelay_OnRemove() : void =
        TitleTextBlock.SetText(BasicMessage("Backpack upgrades "))
        #CurrentBuyOption_textblock.SetVisibility(widget_visibility.Hidden)

            
    Display_Canvas() : void =
        set TimeElapsed = 0
        set UIIsDisplayed = true
        spawn: 
            StartTimeElapsed()
        if:
            Agent:=MaybeAgent?
            PlayerUI:=GetPlayerUI[player[Agent]]
            Canvas := MaybeCanvas?
        then:
            ##Print("Assigning created UI before")
            PlayerUI.AddWidget(Canvas,player_ui_slot{InputMode := ui_input_mode.All, ZOrder:=1}) 
            RemoveDelay_OnAdd() 
            

    StartTimeElapsed()<suspends> : void =
        loop:
            ##Print("IN THE LOOP TimeElapsed: {TimeElapsed}")
            if(TimeElapsed > MaxUIDisplayTime):
                ##Print("Closing the UI because 20 seconds has pased")
                Remove_CanvasWithoutMessage()
                break
            if(UIIsDisplayed = false):
                ##Print("Breaking Time Elapsed loop")
                break
            Sleep(1.0)
            set TimeElapsed = TimeElapsed + 1


    MakeCanvas(): canvas=
        Canvas := canvas:
            Slots:=array:
                #Background black image
                canvas_slot:
                    Anchors:=anchors{Minimum:=vector2{X:=0.5,Y:=0.5},Maximum:=vector2{X:=0.5,Y:=0.5}}
                    ZOrder:=0
                    Alignment:=vector2{X:=0.5,Y:=0.5}
                    Widget:=texture_block:
                        DefaultImage:=Images.BlackBackground
                        DefaultDesiredSize:=vector2{X:=BackgroundBlackUISizeX,Y:=BackgroundBlackUISizeY}
                #Background imageProfileData
                canvas_slot:
                    Anchors:=anchors{Minimum:=vector2{X:=0.5,Y:=0.5},Maximum:=vector2{X:=0.5,Y:=0.5}}
                    ZOrder:=0
                    Offsets:=margin{Left:=0.0,Top:=0.0,Right:=BackgroundImageSizeX,Bottom:=BackgroundImageSizeY} #PositionX, PositionY, SizeX, SizeY
                    Alignment:=vector2{X:=0.5,Y:=0.5}
                    Widget:=texture_block:
                        DefaultImage:=Images.Img_MainBackground
                        DefaultDesiredSize:=vector2{X:=BackgroundImageSizeX,Y:=BackgroundImageSizeY}
                #Main overlay
                canvas_slot:
                    Anchors:=anchors{Minimum:=vector2{X:=0.5,Y:=0.5},Maximum:=vector2{X:=0.5,Y:=0.5}}
                    ZOrder:=0
                    #Offsets:=margin{Left:=1920.0,Top:=0.0,Right:=0.0,Bottom:=1080.0} #PositionX, PositionY, SizeX, SizeY
                    Alignment:=vector2{X:=0.5,Y:=0.5}
                    Widget:=overlay:
                        Slots:=array:
                            overlay_slot:
                                HorizontalAlignment:=horizontal_alignment.Center
                                VerticalAlignment:=vertical_alignment.Center
                                Widget:=stack_box:
                                    Orientation:=orientation.Vertical
                                        Slots:=array:
                                            stack_box_slot: #Sapphires per produc
                                                HorizontalAlignment:=horizontal_alignment.Right
                                                Widget:=stack_box:
                                                    Orientation:=orientation.Horizontal
                                                        Slots:=array:
                                                            stack_box_slot: #Current WATER
                                                                Padding:=margin{Left:=InterColumnLeftPadding,Top:=InterColumnTopPadding,Right:=InterColumnRightPadding,Bottom:=InterColumnBottomPadding}
                                                                Widget:=overlay:
                                                                    Slots:=array:
                                                                        overlay_slot:
                                                                            Widget:=TotalWater_BackgroundTexture
                                                                        overlay_slot:
                                                                            HorizontalAlignment:=horizontal_alignment.Left
                                                                            VerticalAlignment:=vertical_alignment.Center 
                                                                            Padding:=margin{Left:=10.0,Top:=0.0,Right:=0.0,Bottom:=0.0}
                                                                            Widget:=texture_block{DefaultDesiredSize:=vector2{X:=IconSmallSizeX,Y:=IconSmallSizeY}, DefaultImage:=Images.Water_Icon}
                                                                        overlay_slot: 
                                                                            HorizontalAlignment:=horizontal_alignment.Left
                                                                            VerticalAlignment:=vertical_alignment.Center 
                                                                            Padding:=margin{Left:=65.0,Top:=0.0,Right:=0.0,Bottom:=0.0}
                                                                            Widget:=TotalWater_TextBlock
                                                            stack_box_slot: #Current FlOUR
                                                                Padding:=margin{Left:=InterColumnLeftPadding,Top:=InterColumnTopPadding,Right:=InterColumnRightPadding,Bottom:=InterColumnBottomPadding}
                                                                Widget:=overlay:
                                                                    Slots:=array:
                                                                        overlay_slot:
                                                                            Widget:=TotalFlour_BackgroundTexture
                                                                        overlay_slot:
                                                                            HorizontalAlignment:=horizontal_alignment.Left
                                                                            VerticalAlignment:=vertical_alignment.Center 
                                                                            Padding:=margin{Left:=10.0,Top:=0.0,Right:=0.0,Bottom:=0.0}
                                                                            Widget:=texture_block{DefaultDesiredSize:=vector2{X:=IconSmallSizeX,Y:=IconSmallSizeY}, DefaultImage:=Images.Flour_Icon}
                                                                        overlay_slot: 
                                                                            HorizontalAlignment:=horizontal_alignment.Left
                                                                            VerticalAlignment:=vertical_alignment.Center 
                                                                            Padding:=margin{Left:=65.0,Top:=0.0,Right:=0.0,Bottom:=0.0}
                                                                            Widget:=TotalFlour_TextBlock
                                                            stack_box_slot: #Current CHEESE
                                                                Padding:=margin{Left:=InterColumnLeftPadding,Top:=InterColumnTopPadding,Right:=InterColumnRightPadding,Bottom:=InterColumnBottomPadding}
                                                                Widget:=overlay:
                                                                    Slots:=array:
                                                                        overlay_slot:
                                                                            Widget:=TotalCheese_BackgroundTexture
                                                                        overlay_slot:
                                                                            HorizontalAlignment:=horizontal_alignment.Left
                                                                            VerticalAlignment:=vertical_alignment.Center 
                                                                            Padding:=margin{Left:=Icon_LeftMargin}
                                                                            Widget:=texture_block{DefaultDesiredSize:=vector2{X:=IconSmallSizeX,Y:=IconSmallSizeY}, DefaultImage:=Images.Cheese_Icon}
                                                                        overlay_slot: 
                                                                            HorizontalAlignment:=horizontal_alignment.Left
                                                                            VerticalAlignment:=vertical_alignment.Center 
                                                                            Padding:=margin{Left:=TextIcon_LeftMargin}
                                                                            Widget:=TotalCheese_TextBlock
                                                            stack_box_slot: #Current PIMENTOS
                                                                Padding:=margin{Left:=InterColumnLeftPadding,Top:=InterColumnTopPadding,Right:=InterColumnRightPadding,Bottom:=InterColumnBottomPadding}
                                                                Widget:=overlay:
                                                                    Slots:=array:
                                                                        overlay_slot:
                                                                            Widget:=TotalPimentos_BackgroundTexture
                                                                        overlay_slot:
                                                                            HorizontalAlignment:=horizontal_alignment.Left
                                                                            VerticalAlignment:=vertical_alignment.Center 
                                                                            Padding:=margin{Left:=Icon_LeftMargin}
                                                                            Widget:=texture_block{DefaultDesiredSize:=vector2{X:=IconSmallSizeX,Y:=IconSmallSizeY}, DefaultImage:=Images.Pimento_Icon}
                                                                        overlay_slot: 
                                                                            HorizontalAlignment:=horizontal_alignment.Left
                                                                            VerticalAlignment:=vertical_alignment.Center 
                                                                            Padding:=margin{Left:=TextIcon_LeftMargin}
                                                                            Widget:=TotalPimentos_TextBlock
                                                                            <#
                                                            stack_box_slot: #Current MUTANT MATERIAL
                                                                Padding:=margin{Left:=InterColumnLeftPadding,Top:=InterColumnTopPadding,Right:=InterColumnRightPadding,Bottom:=InterColumnBottomPadding}
                                                                Widget:=overlay:
                                                                    Slots:=array:
                                                                        overlay_slot:
                                                                            Widget:=TotalMutantMaterial_BackgroundTexture
                                                                        overlay_slot:
                                                                            HorizontalAlignment:=horizontal_alignment.Left
                                                                            VerticalAlignment:=vertical_alignment.Center 
                                                                            Padding:=margin{Left:=Icon_LeftMargin}
                                                                            Widget:=texture_block{DefaultDesiredSize:=vector2{X:=IconSmallSizeX,Y:=IconSmallSizeY}, DefaultImage:=Images.Flour_Icon}
                                                                        overlay_slot: 
                                                                            HorizontalAlignment:=horizontal_alignment.Left
                                                                            VerticalAlignment:=vertical_alignment.Center 
                                                                            Padding:=margin{Left:=TextIcon_LeftMargin}
                                                                            Widget:=TotalMutantMaterial_TextBlock#>
                                            stack_box_slot: #Backpack perks
                                                Widget:=stack_box:
                                                    Orientation:=orientation.Vertical
                                                        Slots:=array:
                                                            stack_box_slot: #Line 1: Title
                                                                Padding:=margin{Left:=RowLeftPadding,Top:=RowTopPadding,Right:=RowRightPadding,Bottom:=RowBottomPadding}
                                                                HorizontalAlignment:=horizontal_alignment.Fill
                                                                VerticalAlignment:=vertical_alignment.Bottom
                                                                Widget:=overlay:
                                                                    Slots:=array:
                                                                        overlay_slot: 
                                                                            HorizontalAlignment:=horizontal_alignment.Fill
                                                                            VerticalAlignment:=vertical_alignment.Bottom
                                                                            Widget:=texture_block{DefaultDesiredSize:=vector2{X:=HorizontalLineSizeX,Y:=HorizontalLineSizeY},DefaultImage:=Images.Img_HorizontalLine}
                                                                        overlay_slot: 
                                                                            HorizontalAlignment:=horizontal_alignment.Center
                                                                            VerticalAlignment:=vertical_alignment.Center 
                                                                            Widget:=TitleTextBlock
                                                                        overlay_slot: 
                                                                            HorizontalAlignment:=horizontal_alignment.Center
                                                                            VerticalAlignment:=vertical_alignment.Center 
                                                                            Widget:=texture_block{DefaultDesiredSize:=vector2{X:=10.0,Y:=LineSizeY},DefaultImage:=Images.EmptyTexture}
                                                            stack_box_slot: #Line 2: All titles
                                                                Padding:=margin{Left:=RowLeftPadding,Top:=RowTopPadding,Right:=RowRightPadding,Bottom:=RowBottomPadding}
                                                                    Widget:=overlay:
                                                                        Slots:=array:
                                                                            overlay_slot: 
                                                                                HorizontalAlignment:=horizontal_alignment.Fill
                                                                                VerticalAlignment:=vertical_alignment.Center
                                                                                Widget:=texture_block{DefaultDesiredSize:=vector2{X:=0.0,Y:=LineSizeY},DefaultImage:=Images.Img_LineBackground}
                                                                            overlay_slot: 
                                                                                Widget:=stack_box:
                                                                                    Orientation:=orientation.Horizontal
                                                                                        Slots:=array:
                                                                                            stack_box_slot: 
                                                                                                Padding:=margin{Left:=InterColumnLeftPadding,Top:=InterColumnTopPadding,Right:=InterColumnRightPadding,Bottom:=InterColumnBottomPadding}
                                                                                                Widget:=overlay:
                                                                                                    Slots:=array:
                                                                                                        overlay_slot: 
                                                                                                            Widget:=texture_block{DefaultDesiredSize:=vector2{X:=BUColumn1SizeX,Y:=BUColumn1SizeY},DefaultImage:=Images.EmptyTexture}
                                                                                            stack_box_slot: 
                                                                                                Padding:=margin{Left:=InterColumnLeftPadding,Top:=InterColumnTopPadding,Right:=InterColumnRightPadding,Bottom:=InterColumnBottomPadding}
                                                                                                Widget:=overlay:
                                                                                                    Slots:=array:
                                                                                                        overlay_slot:
                                                                                                            Widget:=text_block{DefaultText:=BasicMessage("Resource"), DefaultTextColor:=White, DefaultShadowOffset:=ShadowOffset, DefaultShadowOpacity := DefaultTextShadowOpacity}
                                                                                                        overlay_slot: 
                                                                                                            Widget:=texture_block{DefaultDesiredSize:=vector2{X:=BUColumn2SizeX,Y:=BUColumn2SizeY},DefaultImage:=Images.EmptyTexture}
                                                                                            stack_box_slot: 
                                                                                                Padding:=margin{Left:=InterColumnLeftPadding,Top:=InterColumnTopPadding,Right:=InterColumnRightPadding,Bottom:=InterColumnBottomPadding}
                                                                                                Widget:=overlay:
                                                                                                    Slots:=array:
                                                                                                        overlay_slot:
                                                                                                            Widget:=text_block{DefaultText:=BasicMessage("Current"), DefaultTextColor:=White, DefaultShadowOffset:=ShadowOffset, DefaultShadowOpacity := DefaultTextShadowOpacity}
                                                                                                        overlay_slot: 
                                                                                                            Widget:=texture_block{DefaultDesiredSize:=vector2{X:=BUColumn4SizeX,Y:=BUColumn4SizeY},DefaultImage:=Images.EmptyTexture}
                                                                                            stack_box_slot: 
                                                                                                Padding:=margin{Left:=InterColumnLeftPadding,Top:=InterColumnTopPadding,Right:=InterColumnRightPadding,Bottom:=InterColumnBottomPadding}
                                                                                                Widget:=overlay:
                                                                                                    Slots:=array:
                                                                                                        overlay_slot:
                                                                                                            Widget:=text_block{DefaultText:=BasicMessage("Upgrade"), DefaultTextColor:=White, DefaultShadowOffset:=ShadowOffset, DefaultShadowOpacity := DefaultTextShadowOpacity}
                                                                                                        overlay_slot: 
                                                                                                            Widget:=texture_block{DefaultDesiredSize:=vector2{X:=BUColumn5SizeX,Y:=BUColumn5SizeY},DefaultImage:=Images.EmptyTexture}
                                                                                            stack_box_slot: 
                                                                                                Padding:=margin{Left:=InterColumnLeftPadding,Top:=InterColumnTopPadding,Right:=InterColumnRightPadding,Bottom:=InterColumnBottomPadding}
                                                                                                Widget:=overlay:
                                                                                                    Slots:=array:
                                                                                                        overlay_slot:
                                                                                                            Widget:=text_block{DefaultText:=BasicMessage("Upgrade cost"), DefaultTextColor:=White, DefaultShadowOffset:=ShadowOffset, DefaultShadowOpacity := DefaultTextShadowOpacity}
                                                                                                        overlay_slot: 
                                                                                                            Widget:=texture_block{DefaultDesiredSize:=vector2{X:=BUColumn6SizeX,Y:=BUColumn6SizeY},DefaultImage:=Images.EmptyTexture}


                                                            stack_box_slot: #Line 3: 
                                                                Padding:=margin{Left:=RowLeftPadding,Top:=RowTopPadding,Right:=RowRightPadding,Bottom:=RowBottomPadding}
                                                                    Widget:=overlay:
                                                                        Slots:=array:
                                                                            overlay_slot: 
                                                                                HorizontalAlignment:=horizontal_alignment.Fill
                                                                                VerticalAlignment:=vertical_alignment.Center
                                                                                Widget:=texture_block{DefaultDesiredSize:=vector2{X:=0.0,Y:=LineSizeY},DefaultImage:=Images.Img_LineBackground}
                                                                            overlay_slot: 
                                                                                Widget:=stack_box:
                                                                                    Orientation:=orientation.Horizontal
                                                                                        Slots:=array:
                                                                                            stack_box_slot: 
                                                                                                Padding:=margin{Left:=InterColumnLeftPadding,Top:=InterColumnTopPadding,Right:=InterColumnRightPadding,Bottom:=InterColumnBottomPadding}
                                                                                                Widget:=overlay:
                                                                                                    Slots:=array:
                                                                                                        overlay_slot: 
                                                                                                            Widget:=BU_Row1_Icon
                                                                                            stack_box_slot: 
                                                                                                Padding:=margin{Left:=InterColumnLeftPadding,Top:=InterColumnTopPadding,Right:=InterColumnRightPadding,Bottom:=InterColumnBottomPadding}
                                                                                                Widget:=overlay:
                                                                                                    Slots:=array:
                                                                                                        overlay_slot:
                                                                                                            Widget:=text_block{DefaultText:=BasicMessage("Max. Water"), DefaultTextColor:=White, DefaultShadowOffset:=ShadowOffset, DefaultShadowOpacity := DefaultTextShadowOpacity}
                                                                                                        overlay_slot: 
                                                                                                            Widget:=texture_block{DefaultDesiredSize:=vector2{X:=BUColumn2SizeX,Y:=BUColumn2SizeY},DefaultImage:=Images.EmptyTexture}
                                                                                            stack_box_slot: 
                                                                                                Padding:=margin{Left:=InterColumnLeftPadding,Top:=InterColumnTopPadding,Right:=InterColumnRightPadding,Bottom:=InterColumnBottomPadding}
                                                                                                Widget:=overlay:
                                                                                                    Slots:=array:
                                                                                                        overlay_slot:
                                                                                                            Widget:=BU_Row1_Current_TextBlock
                                                                                                        overlay_slot: 
                                                                                                            Widget:=texture_block{DefaultDesiredSize:=vector2{X:=BUColumn4SizeX,Y:=BUColumn4SizeY},DefaultImage:=Images.EmptyTexture}
                                                                                            stack_box_slot: 
                                                                                                Padding:=margin{Left:=InterColumnLeftPadding,Top:=InterColumnTopPadding,Right:=InterColumnRightPadding,Bottom:=InterColumnBottomPadding}
                                                                                                Widget:=overlay:
                                                                                                    Slots:=array:
                                                                                                        overlay_slot:
                                                                                                            Widget:=BU_Row1_Upgrade_TextBlock
                                                                                                        overlay_slot: 
                                                                                                            Widget:=texture_block{DefaultDesiredSize:=vector2{X:=BUColumn5SizeX,Y:=BUColumn5SizeY},DefaultImage:=Images.EmptyTexture}
                                                                                            stack_box_slot: 
                                                                                                Padding:=margin{Left:=InterColumnLeftPadding,Top:=InterColumnTopPadding,Right:=InterColumnRightPadding,Bottom:=InterColumnBottomPadding}
                                                                                                Widget:=overlay:
                                                                                                    Slots:=array:
                                                                                                        overlay_slot: 
                                                                                                            Widget:=texture_block{DefaultDesiredSize:=vector2{X:=BUColumn6SizeX,Y:=BUColumn6SizeY},DefaultImage:=Images.EmptyTexture}
                                                                                                        overlay_slot: 
                                                                                                            HorizontalAlignment:=horizontal_alignment.Left
                                                                                                            VerticalAlignment:=vertical_alignment.Center 
                                                                                                            Padding:=margin{Left:=65.0,Top:=0.0,Right:=0.0,Bottom:=0.0}
                                                                                                            Widget:=BU_Row1_UpgradeCost_TextBlock 
                                                                                                        overlay_slot: 
                                                                                                            HorizontalAlignment:=horizontal_alignment.Left
                                                                                                            VerticalAlignment:=vertical_alignment.Center 
                                                                                                            Padding:=margin{Left:=10.0,Top:=0.0,Right:=0.0,Bottom:=0.0}
                                                                                                            Widget:=texture_block{DefaultDesiredSize:=vector2{X:=IconSmallSizeX,Y:=IconSmallSizeY}, DefaultImage:=Images.Water_Icon}
                                                                                                        overlay_slot:
                                                                                                            HorizontalAlignment:=horizontal_alignment.Fill
                                                                                                            VerticalAlignment:=vertical_alignment.Fill 
                                                                                                            Widget:=BU_Row1Button
                                                                                                        overlay_slot:
                                                                                                            Widget:=BU_Row1_UpgradeCost_TextBlock
                                                                                            

                                                            stack_box_slot: #Line 4: All titles
                                                                Padding:=margin{Left:=RowLeftPadding,Top:=RowTopPadding,Right:=RowRightPadding,Bottom:=RowBottomPadding}
                                                                    Widget:=overlay:
                                                                        Slots:=array:
                                                                            overlay_slot: 
                                                                                HorizontalAlignment:=horizontal_alignment.Fill
                                                                                VerticalAlignment:=vertical_alignment.Center
                                                                                Widget:=texture_block{DefaultDesiredSize:=vector2{X:=0.0,Y:=LineSizeY},DefaultImage:=Images.Img_LineBackground}
                                                                            overlay_slot: 
                                                                                Widget:=stack_box:
                                                                                    Orientation:=orientation.Horizontal
                                                                                        Slots:=array:
                                                                                            stack_box_slot: 
                                                                                                Padding:=margin{Left:=InterColumnLeftPadding,Top:=InterColumnTopPadding,Right:=InterColumnRightPadding,Bottom:=InterColumnBottomPadding}
                                                                                                Widget:=overlay:
                                                                                                    Slots:=array:
                                                                                                        overlay_slot: 
                                                                                                            Widget:=BU_Row2_Icon
                                                                                            stack_box_slot: 
                                                                                                Padding:=margin{Left:=InterColumnLeftPadding,Top:=InterColumnTopPadding,Right:=InterColumnRightPadding,Bottom:=InterColumnBottomPadding}
                                                                                                Widget:=overlay:
                                                                                                    Slots:=array:
                                                                                                        overlay_slot:
                                                                                                            Widget:=text_block{DefaultText:=BasicMessage("Max. Flour"), DefaultTextColor:=White, DefaultShadowOffset:=ShadowOffset, DefaultShadowOpacity := DefaultTextShadowOpacity}
                                                                                                        overlay_slot: 
                                                                                                            Widget:=texture_block{DefaultDesiredSize:=vector2{X:=BUColumn2SizeX,Y:=BUColumn2SizeY},DefaultImage:=Images.EmptyTexture}
                                                                                            stack_box_slot: 
                                                                                                Padding:=margin{Left:=InterColumnLeftPadding,Top:=InterColumnTopPadding,Right:=InterColumnRightPadding,Bottom:=InterColumnBottomPadding}
                                                                                                Widget:=overlay:
                                                                                                    Slots:=array:
                                                                                                        overlay_slot:
                                                                                                            Widget:=BU_Row2_Current_TextBlock
                                                                                                        overlay_slot: 
                                                                                                            Widget:=texture_block{DefaultDesiredSize:=vector2{X:=BUColumn4SizeX,Y:=BUColumn4SizeY},DefaultImage:=Images.EmptyTexture}
                                                                                            stack_box_slot: 
                                                                                                Padding:=margin{Left:=InterColumnLeftPadding,Top:=InterColumnTopPadding,Right:=InterColumnRightPadding,Bottom:=InterColumnBottomPadding}
                                                                                                Widget:=overlay:
                                                                                                    Slots:=array:
                                                                                                        overlay_slot:
                                                                                                            Widget:=BU_Row2_Upgrade_TextBlock
                                                                                                        overlay_slot: 
                                                                                                            Widget:=texture_block{DefaultDesiredSize:=vector2{X:=BUColumn5SizeX,Y:=BUColumn5SizeY},DefaultImage:=Images.EmptyTexture}
                                                                                            stack_box_slot: 
                                                                                                Padding:=margin{Left:=InterColumnLeftPadding,Top:=InterColumnTopPadding,Right:=InterColumnRightPadding,Bottom:=InterColumnBottomPadding}
                                                                                                Widget:=overlay:
                                                                                                    Slots:=array:
                                                                                                        overlay_slot: 
                                                                                                            Widget:=texture_block{DefaultDesiredSize:=vector2{X:=BUColumn6SizeX,Y:=BUColumn6SizeY},DefaultImage:=Images.EmptyTexture}
                                                                                                        overlay_slot: 
                                                                                                            HorizontalAlignment:=horizontal_alignment.Left
                                                                                                            VerticalAlignment:=vertical_alignment.Center 
                                                                                                            Padding:=margin{Left:=65.0,Top:=0.0,Right:=0.0,Bottom:=0.0}
                                                                                                            Widget:=BU_Row2_UpgradeCost_TextBlock 
                                                                                                        overlay_slot: 
                                                                                                            HorizontalAlignment:=horizontal_alignment.Left
                                                                                                            VerticalAlignment:=vertical_alignment.Center 
                                                                                                            Padding:=margin{Left:=10.0,Top:=0.0,Right:=0.0,Bottom:=0.0}
                                                                                                            Widget:=texture_block{DefaultDesiredSize:=vector2{X:=IconSmallSizeX,Y:=IconSmallSizeY}, DefaultImage:=Images.Flour_Icon}
                                                                                                        overlay_slot:
                                                                                                            HorizontalAlignment:=horizontal_alignment.Fill
                                                                                                            VerticalAlignment:=vertical_alignment.Fill
                                                                                                            Widget:=BU_Row2Button
                                                                                                        overlay_slot:
                                                                                                            Widget:=BU_Row2_UpgradeCost_TextBlock
                                                                            

                                                            stack_box_slot: #Line 5: All titles
                                                                Padding:=margin{Left:=RowLeftPadding,Top:=RowTopPadding,Right:=RowRightPadding,Bottom:=RowBottomPadding}
                                                                    Widget:=overlay:
                                                                        Slots:=array:
                                                                            overlay_slot: 
                                                                                HorizontalAlignment:=horizontal_alignment.Fill
                                                                                VerticalAlignment:=vertical_alignment.Center
                                                                                Widget:=texture_block{DefaultDesiredSize:=vector2{X:=0.0,Y:=LineSizeY},DefaultImage:=Images.Img_LineBackground}
                                                                            overlay_slot: 
                                                                                Widget:=stack_box:
                                                                                    Orientation:=orientation.Horizontal
                                                                                        Slots:=array:
                                                                                            stack_box_slot: 
                                                                                                Padding:=margin{Left:=InterColumnLeftPadding,Top:=InterColumnTopPadding,Right:=InterColumnRightPadding,Bottom:=InterColumnBottomPadding}
                                                                                                Widget:=overlay:
                                                                                                    Slots:=array:
                                                                                                        overlay_slot: 
                                                                                                            Widget:=BU_Row3_Icon
                                                                                            stack_box_slot: 
                                                                                                Padding:=margin{Left:=InterColumnLeftPadding,Top:=InterColumnTopPadding,Right:=InterColumnRightPadding,Bottom:=InterColumnBottomPadding}
                                                                                                Widget:=overlay:
                                                                                                    Slots:=array:
                                                                                                        overlay_slot:
                                                                                                            Widget:=text_block{DefaultText:=BasicMessage("Max. Cheese"), DefaultTextColor:=White, DefaultShadowOffset:=ShadowOffset, DefaultShadowOpacity := DefaultTextShadowOpacity}
                                                                                                        overlay_slot: 
                                                                                                            Widget:=texture_block{DefaultDesiredSize:=vector2{X:=BUColumn2SizeX,Y:=BUColumn2SizeY},DefaultImage:=Images.EmptyTexture}
                                                                                            stack_box_slot: 
                                                                                                Padding:=margin{Left:=InterColumnLeftPadding,Top:=InterColumnTopPadding,Right:=InterColumnRightPadding,Bottom:=InterColumnBottomPadding}
                                                                                                Widget:=overlay:
                                                                                                    Slots:=array:
                                                                                                        overlay_slot:
                                                                                                            Widget:=BU_Row3_Current_TextBlock
                                                                                                        overlay_slot: 
                                                                                                            Widget:=texture_block{DefaultDesiredSize:=vector2{X:=BUColumn4SizeX,Y:=BUColumn4SizeY},DefaultImage:=Images.EmptyTexture}
                                                                                            stack_box_slot: 
                                                                                                Padding:=margin{Left:=InterColumnLeftPadding,Top:=InterColumnTopPadding,Right:=InterColumnRightPadding,Bottom:=InterColumnBottomPadding}
                                                                                                Widget:=overlay:
                                                                                                    Slots:=array:
                                                                                                        overlay_slot:
                                                                                                            Widget:=BU_Row3_Upgrade_TextBlock
                                                                                                        overlay_slot: 
                                                                                                            Widget:=texture_block{DefaultDesiredSize:=vector2{X:=BUColumn5SizeX,Y:=BUColumn5SizeY},DefaultImage:=Images.EmptyTexture}
                                                                                            stack_box_slot: 
                                                                                                Padding:=margin{Left:=InterColumnLeftPadding,Top:=InterColumnTopPadding,Right:=InterColumnRightPadding,Bottom:=InterColumnBottomPadding}
                                                                                                Widget:=overlay:
                                                                                                    Slots:=array:
                                                                                                        overlay_slot: 
                                                                                                            Widget:=texture_block{DefaultDesiredSize:=vector2{X:=BUColumn6SizeX,Y:=BUColumn6SizeY},DefaultImage:=Images.EmptyTexture}
                                                                                                        overlay_slot: 
                                                                                                            HorizontalAlignment:=horizontal_alignment.Left
                                                                                                            VerticalAlignment:=vertical_alignment.Center 
                                                                                                            Padding:=margin{Left:=65.0,Top:=0.0,Right:=0.0,Bottom:=0.0}
                                                                                                            Widget:=BU_Row3_UpgradeCost_TextBlock 
                                                                                                        overlay_slot: 
                                                                                                            HorizontalAlignment:=horizontal_alignment.Left
                                                                                                            VerticalAlignment:=vertical_alignment.Center 
                                                                                                            Padding:=margin{Left:=10.0,Top:=0.0,Right:=0.0,Bottom:=0.0}
                                                                                                            Widget:=texture_block{DefaultDesiredSize:=vector2{X:=IconSmallSizeX,Y:=IconSmallSizeY}, DefaultImage:=Images.Cheese_Icon}
                                                                                                        overlay_slot:
                                                                                                            HorizontalAlignment:=horizontal_alignment.Fill
                                                                                                            VerticalAlignment:=vertical_alignment.Fill 
                                                                                                            Widget:=BU_Row3Button
                                                                                                        overlay_slot:
                                                                                                            Widget:=BU_Row3_UpgradeCost_TextBlock
                                                                                                

                                                            stack_box_slot: #Line 6: All titles
                                                                Padding:=margin{Left:=RowLeftPadding,Top:=RowTopPadding,Right:=RowRightPadding,Bottom:=RowBottomPadding}
                                                                    Widget:=overlay:
                                                                        Slots:=array:
                                                                            overlay_slot: 
                                                                                HorizontalAlignment:=horizontal_alignment.Fill
                                                                                VerticalAlignment:=vertical_alignment.Center
                                                                                Widget:=texture_block{DefaultDesiredSize:=vector2{X:=0.0,Y:=LineSizeY},DefaultImage:=Images.Img_LineBackground}
                                                                            overlay_slot: 
                                                                                Widget:=stack_box:
                                                                                    Orientation:=orientation.Horizontal
                                                                                        Slots:=array:
                                                                                            stack_box_slot: 
                                                                                                Padding:=margin{Left:=InterColumnLeftPadding,Top:=InterColumnTopPadding,Right:=InterColumnRightPadding,Bottom:=InterColumnBottomPadding}
                                                                                                Widget:=overlay:
                                                                                                    Slots:=array:
                                                                                                        overlay_slot: 
                                                                                                            Widget:=BU_Row4_Icon
                                                                                            stack_box_slot: 
                                                                                                Padding:=margin{Left:=InterColumnLeftPadding,Top:=InterColumnTopPadding,Right:=InterColumnRightPadding,Bottom:=InterColumnBottomPadding}
                                                                                                Widget:=overlay:
                                                                                                    Slots:=array:
                                                                                                        overlay_slot:
                                                                                                            Widget:=text_block{DefaultText:=BasicMessage("Max. Pimentos"), DefaultTextColor:=White, DefaultShadowOffset:=ShadowOffset, DefaultShadowOpacity := DefaultTextShadowOpacity}
                                                                                                        overlay_slot: 
                                                                                                            Widget:=texture_block{DefaultDesiredSize:=vector2{X:=BUColumn2SizeX,Y:=BUColumn2SizeY},DefaultImage:=Images.EmptyTexture}
                                                                                            stack_box_slot: 
                                                                                                Padding:=margin{Left:=InterColumnLeftPadding,Top:=InterColumnTopPadding,Right:=InterColumnRightPadding,Bottom:=InterColumnBottomPadding}
                                                                                                Widget:=overlay:
                                                                                                    Slots:=array:
                                                                                                        overlay_slot:
                                                                                                            Widget:=BU_Row4_Current_TextBlock
                                                                                                        overlay_slot: 
                                                                                                            Widget:=texture_block{DefaultDesiredSize:=vector2{X:=BUColumn4SizeX,Y:=BUColumn4SizeY},DefaultImage:=Images.EmptyTexture}
                                                                                            stack_box_slot: 
                                                                                                Padding:=margin{Left:=InterColumnLeftPadding,Top:=InterColumnTopPadding,Right:=InterColumnRightPadding,Bottom:=InterColumnBottomPadding}
                                                                                                Widget:=overlay:
                                                                                                    Slots:=array:
                                                                                                        overlay_slot:
                                                                                                            Widget:=BU_Row4_Upgrade_TextBlock
                                                                                                        overlay_slot: 
                                                                                                            Widget:=texture_block{DefaultDesiredSize:=vector2{X:=BUColumn5SizeX,Y:=BUColumn5SizeY},DefaultImage:=Images.EmptyTexture}
                                                                                            stack_box_slot: 
                                                                                                Padding:=margin{Left:=InterColumnLeftPadding,Top:=InterColumnTopPadding,Right:=InterColumnRightPadding,Bottom:=InterColumnBottomPadding}
                                                                                                Widget:=overlay:
                                                                                                    Slots:=array:
                                                                                                        overlay_slot: 
                                                                                                            Widget:=texture_block{DefaultDesiredSize:=vector2{X:=BUColumn6SizeX,Y:=BUColumn6SizeY},DefaultImage:=Images.EmptyTexture}
                                                                                                        overlay_slot: 
                                                                                                            HorizontalAlignment:=horizontal_alignment.Left
                                                                                                            VerticalAlignment:=vertical_alignment.Center 
                                                                                                            Padding:=margin{Left:=65.0,Top:=0.0,Right:=0.0,Bottom:=0.0}
                                                                                                            Widget:=BU_Row4_UpgradeCost_TextBlock 
                                                                                                        overlay_slot: 
                                                                                                            HorizontalAlignment:=horizontal_alignment.Left
                                                                                                            VerticalAlignment:=vertical_alignment.Center 
                                                                                                            Padding:=margin{Left:=10.0,Top:=0.0,Right:=0.0,Bottom:=0.0}
                                                                                                            Widget:=texture_block{DefaultDesiredSize:=vector2{X:=IconSmallSizeX,Y:=IconSmallSizeY}, DefaultImage:=Images.Pimento_Icon}
                                                                                                        overlay_slot:
                                                                                                            HorizontalAlignment:=horizontal_alignment.Fill
                                                                                                            VerticalAlignment:=vertical_alignment.Fill
                                                                                                            Widget:=BU_Row4Button
                                                                                                        overlay_slot:
                                                                                                            Widget:=BU_Row4_UpgradeCost_TextBlock
                                                            <#stack_box_slot: #Mutant Material
                                                                Padding:=margin{Left:=RowLeftPadding,Top:=RowTopPadding,Right:=RowRightPadding,Bottom:=RowBottomPadding}
                                                                    Widget:=overlay:
                                                                        Slots:=array:
                                                                            overlay_slot: 
                                                                                HorizontalAlignment:=horizontal_alignment.Fill
                                                                                VerticalAlignment:=vertical_alignment.Center
                                                                                Widget:=texture_block{DefaultDesiredSize:=vector2{X:=0.0,Y:=LineSizeY},DefaultImage:=Images.Img_LineBackground}
                                                                            overlay_slot: 
                                                                                Widget:=stack_box:
                                                                                    Orientation:=orientation.Horizontal
                                                                                        Slots:=array:
                                                                                            stack_box_slot: 
                                                                                                Padding:=margin{Left:=InterColumnLeftPadding,Top:=InterColumnTopPadding,Right:=InterColumnRightPadding,Bottom:=InterColumnBottomPadding}
                                                                                                Widget:=overlay:
                                                                                                    Slots:=array:
                                                                                                        overlay_slot: 
                                                                                                            Widget:=BU_Row5_Icon
                                                                                            stack_box_slot: 
                                                                                                Padding:=margin{Left:=InterColumnLeftPadding,Top:=InterColumnTopPadding,Right:=InterColumnRightPadding,Bottom:=InterColumnBottomPadding}
                                                                                                Widget:=overlay:
                                                                                                    Slots:=array:
                                                                                                        overlay_slot:
                                                                                                            Widget:=text_block{DefaultText:=BasicMessage("Max. Mutant Material"), DefaultTextColor:=White, DefaultShadowOffset:=ShadowOffset, DefaultShadowOpacity := DefaultTextShadowOpacity}
                                                                                                        overlay_slot: 
                                                                                                            Widget:=texture_block{DefaultDesiredSize:=vector2{X:=BUColumn2SizeX,Y:=BUColumn2SizeY},DefaultImage:=Images.EmptyTexture}
                                                                                            stack_box_slot: 
                                                                                                Padding:=margin{Left:=InterColumnLeftPadding,Top:=InterColumnTopPadding,Right:=InterColumnRightPadding,Bottom:=InterColumnBottomPadding}
                                                                                                Widget:=overlay:
                                                                                                    Slots:=array:
                                                                                                        overlay_slot:
                                                                                                            Widget:=BU_Row5_Current_TextBlock
                                                                                                        overlay_slot: 
                                                                                                            Widget:=texture_block{DefaultDesiredSize:=vector2{X:=BUColumn4SizeX,Y:=BUColumn4SizeY},DefaultImage:=Images.EmptyTexture}
                                                                                            stack_box_slot: 
                                                                                                Padding:=margin{Left:=InterColumnLeftPadding,Top:=InterColumnTopPadding,Right:=InterColumnRightPadding,Bottom:=InterColumnBottomPadding}
                                                                                                Widget:=overlay:
                                                                                                    Slots:=array:
                                                                                                        overlay_slot:
                                                                                                            Widget:=BU_Row5_Upgrade_TextBlock
                                                                                                        overlay_slot: 
                                                                                                            Widget:=texture_block{DefaultDesiredSize:=vector2{X:=BUColumn5SizeX,Y:=BUColumn5SizeY},DefaultImage:=Images.EmptyTexture}
                                                                                            stack_box_slot: 
                                                                                                Padding:=margin{Left:=InterColumnLeftPadding,Top:=InterColumnTopPadding,Right:=InterColumnRightPadding,Bottom:=InterColumnBottomPadding}
                                                                                                Widget:=overlay:
                                                                                                    Slots:=array:
                                                                                                        overlay_slot: 
                                                                                                            Widget:=texture_block{DefaultDesiredSize:=vector2{X:=BUColumn6SizeX,Y:=BUColumn6SizeY},DefaultImage:=Images.EmptyTexture}
                                                                                                        overlay_slot: 
                                                                                                            HorizontalAlignment:=horizontal_alignment.Left
                                                                                                            VerticalAlignment:=vertical_alignment.Center 
                                                                                                            Padding:=margin{Left:=65.0,Top:=0.0,Right:=0.0,Bottom:=0.0}
                                                                                                            Widget:=BU_Row5_UpgradeCost_TextBlock 
                                                                                                        overlay_slot: 
                                                                                                            HorizontalAlignment:=horizontal_alignment.Left
                                                                                                            VerticalAlignment:=vertical_alignment.Center 
                                                                                                            Padding:=margin{Left:=10.0,Top:=0.0,Right:=0.0,Bottom:=0.0}
                                                                                                            Widget:=texture_block{DefaultDesiredSize:=vector2{X:=IconSmallSizeX,Y:=IconSmallSizeY}, DefaultImage:=Images.Mutant_Material_Icon}
                                                                                                        overlay_slot:
                                                                                                            HorizontalAlignment:=horizontal_alignment.Fill
                                                                                                            VerticalAlignment:=vertical_alignment.Center  
                                                                                                            Widget:=BU_Row5Button
                                                                                                        overlay_slot:
                                                                                                            Widget:=BU_Row5_UpgradeCost_TextBlock#>
                                                            stack_box_slot: #Line 8: All titles
                                                                Padding:=margin{Left:=RowLeftPadding,Top:=RowTopPadding,Right:=RowRightPadding,Bottom:=RowBottomPadding}
                                                                    Widget:=overlay:
                                                                        Slots:=array:
                                                                            overlay_slot: 
                                                                                HorizontalAlignment:=horizontal_alignment.Center
                                                                                VerticalAlignment:=vertical_alignment.Center
                                                                                Widget:=stack_box:
                                                                                    Orientation:=orientation.Horizontal
                                                                                        Slots:=array:
                                                                                            stack_box_slot: 
                                                                                                Padding:=margin{Left:=InterColumnLeftPadding,Top:=InterColumnTopPadding,Right:=InterColumnRightPadding,Bottom:=InterColumnBottomPadding}
                                                                                                Widget:=overlay:
                                                                                                    Slots:=array:
                                                                                                        overlay_slot: 
                                                                                                            Widget:=texture_block{DefaultDesiredSize:=vector2{X:=BigCloseButtonSizeX,Y:=BigCloseButtonSizeY},DefaultImage:=Images.EmptyTexture}
                                                                                                        overlay_slot: 
                                                                                                            HorizontalAlignment:=horizontal_alignment.Fill
                                                                                                            VerticalAlignment:=vertical_alignment.Fill
                                                                                                            Padding:=margin{Left:=InterColumnLeftPadding,Top:=InterColumnTopPadding,Right:=InterColumnRightPadding,Bottom:=InterColumnBottomPadding}
                                                                                                            Widget:=CloseButton
                                                            
MakeBackpackUpgradeUI<constructor>(
    InBakeryMachinesManager : bakery_machines_manager)<transacts> := backpack_upgrades_ui :
        BakeryMachinesManager := InBakeryMachinesManager
                                                                                                               