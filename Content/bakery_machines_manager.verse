using { /Fortnite.com/Devices }
using { /Verse.org/Simulation }
using { /UnrealEngine.com/Temporary/Diagnostics }
using { /Verse.org/Assets }
using { /UnrealEngine.com/Temporary/SpatialMath }
using { /Verse.org/Random }

# See https://dev.epicgames.com/documentation/en-us/uefn/create-your-own-device-in-verse for how to create a verse device.

# A Verse-authored creative device that can be placed in a level
bakery_machines_manager := class():

    GameManager : game_manager
    PetsManager : pets_manager
    HexaGameManager : hexa_minigame_manager

    var MaybeAgentOwner : ?agent = false
    var IsBakerReady : logic = false
    var MaybeBB_SuperEneryProp : ?creative_prop = false
    var MaybeCBB_SuperEneryProp : ?creative_prop = false
    var MaybePCSB_SuperEneryProp : ?creative_prop = false
    var MaybeUKTutorial : ?uk_tutorial = false
    var IsOnUKTutorial : logic = false
    
    #Class settings

    #UPGRADE COSTS
    #-------------------
    #https://es.symbolab.com/graphing-calculator
    #https://docs.google.com/spreadsheets/d/1G1X-YGXF_2jjdY69GHdhqIrsZMOHPdfXblED1hLdOEU/edit#gid=0

    #Materials
    Mat_NormalPipe : material = Materials.Mat_PipeStateNotColor
    Mat_PipeState0 : material = Materials.Mat_PipeState0
    Mat_PipeState1 : material = Materials.Mat_PipeState1
    Mat_PipeState2 : material = Materials.Mat_PipeState2

    Mat_RedScreen : material = Materials.RedScreen_Inst
    Mat_GreenScreen : material = Materials.GreenScreen_Inst

    Mat_RedBakerState : material = Materials.MI_Complement_red
    Mat_YellowBakerState : material = Materials.MI_Complement_Yellow
    Mat_GreenBakerState : material = Materials.MI_Complement_green

#Async Fame
    var AsyncFame : float = -1.0
    var AsyncFameDifference : float = -1.0

#Special effects

#OWNER SETTINGS
#Manages and updates stats for each player.
    var PlayersDataManager : player_data_manager = player_data_manager{}
    var PlayerDataCache : player_data_table = player_data_table{}
    #Options
    #var IsMusicOn : logic = true
    #var IsQuestTrackerOn : logic = true

#Agents UI
    var HUDMessage_PlayerWarnings : hud_message_device = hud_message_device{}
    var HUDMessage_PlayerInventory : hud_message_device = hud_message_device{}
    var AgentAndBaker_Index : int = -1
    var OwnerReferenceDevice : player_reference_device = player_reference_device{}

    #UIs
    var MaybeMasteryUI : ?mastery_ui = false
    var MaybeActionInfoUI : ?action_info_ui = false
    var MaybeDungeonUI : ?dungeon_ui = false
    var MaybeMainInventoryUI : ?main_inventory_ui = false
    var MaybeEntityHealthUI : ?entity_health_bar_ui = false
    var MaybePickaxeUpgradesUI : ?pickaxe_upgrades_ui = false
    #var MaybeResourcersUI : ?resourcers_ui = false
    var MaybeMessagesUI : ?messages_ui = false

    var MaybePCSBakersUpgradesUI : ?PCSbaker_upgrades_ui = false
    var MaybeBBakersUpgradesUI : ?Bbaker_upgrades_ui = false
    var MaybeCBBakersUpgradesUI : ?CBbaker_upgrades_ui = false

    var MaybePCSBInfoUI : ?pcsb_info_ui = false
    var MaybeCBBInfoUI : ?cbb_info_ui = false
    var MaybeBBInfoUI : ?bb_info_ui = false

    var MaybeDroneUpgrades : ?drone_upgrades_ui = false
    var MaybeBackpackUpgradesUI : ?backpack_upgrades_ui = false
    var MaybePickaxePerksUI : ?pickaxe_perks_upgrades_ui = false

    var MaybeDepositsUpgradeUI : ?deposits_upgrades_ui = false
    var MaybeFastTravelUI : ?fast_travel_ui = false

    var MaybeWeaponsUpgradesUI : ?weapons_upgrades_ui = false
    var MaybeDropCoalUI : ?drop_coal_ui = false
    var MaybeOptionsMenuUI : ?options_menu_ui = false
    var MaybeLanguageUI : ?language_ui = false
    var MaybeTokensUI : ?tokens_ui = false

    var BBInfoIsDisplayed : logic = false
    var CBBInfoIsDisplayed : logic = false 
    var PCSBInfoIsDisplayed : logic = false 



#Super Energy
    #var SuperEnergy_Owned : float = 0.0

#Teleporter
    var Prop_PersonalTeleporter : creative_prop = creative_prop{}
    var Button_PersonalTeleporter : button_device = button_device{}

    #AFKRegister    
    var AFKRegister : float = 0.0

#Backpack limits
#Backpack / inventory
    var AllBackpackLimits : []float = array{0.0,0.0,0.0,0.0}

    var AllAgentsMutantMaterialBackpackLimit : float = 0.0
    var AllAgentsCoalBackpackLimit : float = 0.0


#Pickaxe stats
    var PickaxeTotalStrength : float = 0.0
    var Secret_PickaxeEfficiency : float = 1.0
    var AllPickaxeEfficiency : []float = array{0.0,0.0,0.0,0.0,0.0}

    
    var PickaxeCriticalProb : float = 0.0
    var PickaxeCriticalDamage : float = 0.0
    var PickaxeCoalLuck : float = 1.0
    var PickaxeWeakPointMult : float = 0.0
    var PickaxeMultiResource : int = 0

#Currencies
        #Super energy
        
    var OverPoweredSign : creative_prop = creative_prop{}
    var SuperEnergy_Multiplier : float = 1.0

        #Rewards per products

    Base_SapphiresPerB : float = (Base_Req_WaterFor_B+Base_Req_FlourFor_B)*1.0/Req_Resourcers_PerSapphire
    Base_SapphiresPerCB : float = (Base_Req_WaterFor_CB+Base_Req_FlourFor_CB+Base_Req_CheeseFor_CB)*1.0/Req_Resourcers_PerSapphire
    Base_SapphiresPerPCS : float = (Base_Req_WaterFor_PCS+Base_Req_FlourFor_PCS+Base_Req_CheeseFor_PCS+Base_Req_PimentosFor_PCS)*1.0/Req_Resourcers_PerSapphire

    SapphiresPerB : float = (Base_Req_WaterFor_B+Base_Req_FlourFor_B)*1.0/Req_Resourcers_PerSapphire
    SapphiresPerCB : float = (Base_Req_WaterFor_CB+Base_Req_FlourFor_CB+Base_Req_CheeseFor_CB)*1.0/Req_Resourcers_PerSapphire
    SapphiresPerPCS : float = (Base_Req_WaterFor_PCS+Base_Req_FlourFor_PCS+Base_Req_CheeseFor_PCS+Base_Req_PimentosFor_PCS)*1.0/Req_Resourcers_PerSapphire

#Timers
    var SuperEnergyRemaining_Timer : timer_device = timer_device{}

#Drones
        #Drone stats
    var LV_DroneSpeed : int = 0
    var LV_DroneStorage : int = 0

    #var B_On_Drone : float = 0.0
    #var CB_On_Drone : float = 0.0
    #var PCS_On_Drone : float = 0.0

    #var MaxB_Storage : float = Base_B_Storage
    #var MaxCB_Storage : float = Base_CB_Storage
    #var MaxPCS_Storage : float = Base_PCS_Storage

    var AllLV_SalesEfficiency : []int = array{0,0,0}
    var All_SalesEfficiency : []float = array{0.0,0.0,0.0}

    var DroneSpeed : int = Base_Drone_Speed
    var TimeToDroneReturn : int = 0
    var IsDronePropOnBase : logic = true


#Auto Pipes
    #var AllLV_PasiveResourcers : [] int = array{0,0,0,0}
    var AllPasiveResourcers : []float = array{0.0,0.0,0.0,0.0}

#Main deposits
    var Button_ResFunnel : button_device = button_device{}

    var AllLV_MainDeposits : [] int = array{0,0,0,0}
    var AllMainDeposit_Limits : []float = array{1000.0,1000.0,1000.0,1000.0}

    #Billboards
    var WaterDeposit_Billbord : billboard_device = billboard_device{}
    var FlourDeposit_Billbord : billboard_device = billboard_device{}
    var CheeseDeposit_Billbord : billboard_device = billboard_device{}
    var PimentosDeposit_Billbord : billboard_device = billboard_device{}
    var DroneDeposit_Billbord : billboard_device = billboard_device{}
    var DroneShip_Billbord : billboard_device = billboard_device{}


    #var MainWaterDeposit_Value : float = 0.0
    #var MainFlourDeposit_Value : float = 0.0
    #var MainCheeseDeposit_Value : float = 0.0
    #var MainPimentosDeposit_Value : float = 0.0

    var MainMutantMaterialDeposit_Value : float = 0.0
        #Props
    var MainWaterDeposit_AmmountProp : creative_prop = creative_prop{}
    var MainFlourDeposit_AmmountProp : creative_prop = creative_prop{}
    var MainCheeseDeposit_AmmountProp : creative_prop = creative_prop{}
    var MainPimentosDeposit_AmmountProp : creative_prop = creative_prop{}

#Baker Machines
    var WarningMessageDisplayed : logic = false
    var IsCBBAvailable : logic = false
    var IsPCSBAvailable : logic = false
        #Props
    var BB_Prop : creative_prop = creative_prop{}
    var CBB_Prop : creative_prop = creative_prop{}
    var PCSB_Prop : creative_prop = creative_prop{}
        #Progress
    var Progress_ForB : float = 0.0
    var Progress_ForCB : float = 0.0
    var Progress_ForPCS : float = 0.0
    
        #Level of production
    #var TLV_BB : []int = array{0,-1,-1,-1,-1,-1,-1,-1,-1,-1,-1,-1,-1,-1,-1,-1,-1,-1,-1,-1}
    #var TLV_CBB : []int = array{0,-1,-1,-1,-1,-1,-1,-1,-1,-1,-1,-1,-1,-1,-1,-1,-1,-1,-1,-1}
    #var TLV_PCSB : []int = array{0,-1,-1,-1,-1,-1,-1,-1,-1,-1,-1,-1,-1,-1,-1,-1,-1,-1,-1,-1}
        #Ammount produced
    var B_BachProduction : float = 0.0
    var CB_BachProduction : float = 0.0
    var PCS_BachProduction : float = 0.0
    var BB_Production : float = 0.0
    var CBB_Production : float = 0.0
    var PCSB_Production : float = 0.0
        #Speed
    var BB_Speed : float = Base_BB_Speed
    var CBB_Speed : float = Base_CBB_Speed
    var PCSB_Speed : float = Base_PCSB_Speed
        #Efficiency
    var BB_WaterEfficiency: float = 2.5
    var CBB_WaterEfficiency: float = 2.5
    var PCSB_WaterEfficiency: float = 2.5
    var BB_FlourEfficiency: float = 2.5
    var CBB_FlourEfficiency: float = 2.5
    var PCSB_FlourEfficiency: float = 2.5
    var BB_CheeseEfficiency: float = 2.5
    var CBB_CheeseEfficiency: float = 2.5
    var PCSB_CheeseEfficiency: float = 2.5
    var PCSB_PimentosEfficiency: float = 2.5

#Subdeposits
        #Requirements
    var Req_WaterFor_B : float = 999.0
    var Req_FlourFor_B : float = 999.0
    var Req_WaterFor_CB : float = 999.0
    var Req_FlourFor_CB : float = 999.0
    var Req_CheeseFor_CB : float = 999.0
    var Req_WaterFor_PCS : float = 999.0
    var Req_FlourFor_PCS : float = 999.0
    var Req_CheeseFor_PCS : float = 999.0
    var Req_PimentosFor_PCS : float = 999.0
    var Max_PimentosFor_Pets : float = 999.0
        #AmmountValues

#Valves
       #Buttons
    var AllButtons_Valves : []button_device = array{}
        #Simplified Buttons
    var Simplified_ValveButtons_B : button_device = button_device{}
    var Simplified_ValveButtons_CB : button_device = button_device{}
    var Simplified_ValveButtons_PCS : button_device = button_device{}
        #Props
    var AllProps_Valves : []creative_prop = array{}
        #States
    var ValveState_WaterFor_B : float = 0.0
    var ValveState_FlourFor_B : float = 0.0

    var ValveState_WaterFor_CB : float = 0.0
    var ValveState_FlourFor_CB : float = 0.0
    var ValveState_CheeseFor_CB  : float = 0.0

    var ValveState_WaterFor_PCS : float = 0.0
    var ValveState_FlourFor_PCS : float = 0.0
    var ValveState_CheeseFor_PCS : float = 0.0
    var ValveState_PimentosFor_PCS : float = 1.0
    #var ValveState_PimentosFor_Pets : float = 0.0

    var ValveState_WaterFor_Pet : int = 0
    var ValveState_FlourFor_Pet : int = 0
    var ValveState_CheeseFor_Pet : int = 0
    #var ValveState_PimentosFor_Pet : int = 0
        #Flows
    var WaterAmmount_ForB_ToTake : float = 0.0
    var WaterAmmount_ForCB_ToTake : float = 0.0
    var WaterAmmount_ForPCS_ToTake : float = 0.0
    var FlourAmmount_ForB_ToTake : float = 0.0
    var FlourAmmount_ForCB_ToTake : float = 0.0
    var FlourAmmount_ForPCS_ToTake : float = 0.0
    var CheeseAmmount_ForCB_ToTake : float = 0.0
    var CheeseAmmount_ForPCS_ToTake : float = 0.0
    var PimentosAmmount_ForPCS_ToTake : float = 0.0
    #var PimentosAmmount_ForPets_ToTake : float = 0.0
        #Flow Per Second
    var WaterFlowPerSecond_ForB_ToTake : float = 0.0
    var WaterFlowPerSecond_ForCB_ToTake : float = 0.0
    var WaterFlowPerSecond_ForPCS_ToTake : float = 0.0
    var FlourFlowPerSecond_ForB_ToTake : float = 0.0
    var FlourFlowPerSecond_ForCB_ToTake : float = 0.0
    var FlourFlowPerSecond_ForPCS_ToTake : float = 0.0
    var CheeseFlowPerSecond_ForCB_ToTake : float = 0.0
    var CheeseFlowPerSecond_ForPCS_ToTake : float = 0.0
    var PimentosFlowPerSecond_ForPCS_ToTake : float = 0.0
    #var PimentosFlowPerSecond_ForPets_ToTake : float = 0.0
        

#Computers
    var Button_DepositsUpgrades : button_device = button_device{}

    var Button_InfoBB : button_device = button_device{}
    var Button_InfoCBB : button_device = button_device{}
    var Button_InfoPCSB : button_device = button_device{}

    var Button_BBUpgrades : button_device = button_device{}
    var Button_CBBUpgrades : button_device = button_device{}
    var Button_PCSBUpgrades : button_device = button_device{}

#Drones
    var DroneProp : creative_prop = creative_prop{}

    var DroneInitialTransform : transform = transform{}
    var FinalPosition : vector3 = vector3{}

    var Button_DroneUpgrades : button_device = button_device{}
    var Button_DroneShip : button_device = button_device{}

    #String and Messages
    #Strings
    var VALVEFLOW_BUTTON_WATER_MESSAGE : string = "💧 Flow: "
    var VALVEFLOW_BUTTON_FLOUR_MESSAGE : string = "🌾 Flow: "
    var VALVEFLOW_BUTTON_CHEESE_MESSAGE : string = "🧀 Flow: "
    var VALVEFLOW_BUTTON_PIMENTOS_MESSAGE : string = "🌶️ Flow: "

    #UKTutorialController
    SetUKTutorial(InTutorial : logic) : void =
        set IsOnUKTutorial = InTutorial

    SetUKTutorial(InTutorial : uk_tutorial) : void =
        set MaybeUKTutorial = option{InTutorial}

    SetUKQuestDescription(MessageDescription : message) : void =
        if(IsOnUKTutorial?):
            Show_QuestPanel()
            if(MainInventoryUI:=MaybeMainInventoryUI?):
                MainInventoryUI.SetUKQuest(MessageDescription)

    HideUKQuestDescription() : void =
        if(IsOnUKTutorial?,MainInventoryUI:=MaybeMainInventoryUI?):
            MainInventoryUI.Hide_QuestPanel()
            MainInventoryUI.Hide_OrderPanel()

    Show_QuestPanel() : void =
        if(MainInventoryUI:=MaybeMainInventoryUI?):
            MainInventoryUI.Show_QuestPanel()

    SetUKQuest() : void =
        if(MainInventoryUI:=MaybeMainInventoryUI?):
            MainInventoryUI.DisplayQuest(PlayerDataCache)

    DisplayQuest() : void =
        if(MainInventoryUI:=MaybeMainInventoryUI?):
            MainInventoryUI.DisplayQuest(PlayerDataCache)

    #AsyncFame
    SetAsyncNewFame(NewFame : float, Difference : float) : void =
        set AsyncFame = NewFame
        set AsyncFameDifference = Difference
        Log("AsyncFame: {AsyncFame}, Difference: {Difference}",LOG_LEVEL_DEBUG6)

    CheckAsyncFame() : void =
        if(AsyncFame>=0.0):
            DisplayResourcersGather(array{AsyncFameDifference},array{FAME},false)
            SetNewPlayerDataCache(PlayersDataManager.SetFame(PlayerDataCache,AsyncFame))
            set AsyncFame = -1.0
            set AsyncFameDifference = -1.0
            RefreshAllUI()

    SetMusicState(NewState : logic) : void =
        SetNewPlayerDataCache(PlayersDataManager.SetMusicState(PlayerDataCache,NewState))

    SetQuestTrackerState(NewState : logic) : void =
        SetNewPlayerDataCache(PlayersDataManager.SetTrackerState(PlayerDataCache,NewState))

    
    SetNewPlayerDataCache(NewCache : player_data_table) : void =
        if(AgentOwner:=MaybeAgentOwner?, Player := player[AgentOwner], Player.IsActive[]):
            set PlayerDataCache = NewCache

    DecrementMiniGamesCooldown<private>() : void =
        SetNewPlayerDataCache(PlayersDataManager.DecrementAllMiniGameCooldown(PlayerDataCache, 1))

    CheckQuestTracker() : void =
        if(Agent := MaybeAgentOwner?, Player := player[Agent], Player.IsActive[]):
            IndexToUse:=GameManager.GetAgentIndex(Agent)
            if(PlayerDataCache.IsQuestTrackerOn?<#, PlayerData:=PlayersDataManager.GetPlayerData[Player]#>):
                TupleResult:=GameManager.QuestsAndOrdersDatabase.GetQuestObjetiveTransform(<#PlayerDataCache.QuestAssigned_NumQuest#>PlayerDataCache.NumQuestsCompleted,PlayerDataCache.QuestAssigned_ObjetiveType,Agent,PlayerDataCache,PlayerDataCache.QuestAssigned_NumQuest)
                IsValid:=TupleResult(0)
                TransFormObjetive:=TupleResult(1)
                if(IsValid?, not IsOnUKTutorial?, MapBeaconDevice:=GameManager.AllMapBeaconDevices[IndexToUse]):
                    GameManager.VFXManager.SpawnQuestTracker(Agent,TransFormObjetive,IndexToUse,GameManager.TotalGameTime)
                    if(10.0<Distance(MapBeaconDevice.GetTransform().Translation,TransFormObjetive.Translation)):
                        if(MapBeaconDevice.TeleportTo[TransFormObjetive]):
                        PrintWarning("On CheckQuestTracker, Beacon teleported to the correct position")

    CheckMusicPlay(MusicType : int) : void =
        if(Agent := MaybeAgentOwner?):
            var AllMusicsToStop : []int = array{DUNGEONMUSIC_SOUND,DUNGEONMINIBOSSMUSIC_SOUND,HEXAGAMEMUSIC_SOUND,BAKERSMUSIC_SOUND,DESERTMUSIC_SOUND,SECRETMUSIC_SOUND}
            #Removing music that is going to be played
            for(Index:=0..AllMusicsToStop.Length-1, MusicToStop := AllMusicsToStop[Index], MusicType=MusicToStop, NewArray := AllMusicsToStop.RemoveElement[Index]):
                set AllMusicsToStop = NewArray
            #Stoping all music, except the one to play
            for(MusicToStop : AllMusicsToStop):
                spawn:
                    GameManager.VFXManager.StopSound(MusicToStop,0.0,option{Agent},false)
            
            #Play Music Only if the players decides so
            if(PlayerDataCache.IsMusicOn?):
                #Now, play the selected music
                Log("Playing music: {MusicType}",LOG_LEVEL_DEBUG6)
                spawn:
                    GameManager.VFXManager.PlaySound(MusicType,0.0,option{Agent},false)
            else:
                set AllMusicsToStop = array{DUNGEONMUSIC_SOUND,HEXAGAMEMUSIC_SOUND,BAKERSMUSIC_SOUND,DESERTMUSIC_SOUND,SECRETMUSIC_SOUND}
                for(Index:=0..AllMusicsToStop.Length-1, MusicToStop := AllMusicsToStop[Index]):
                    spawn:
                        GameManager.VFXManager.StopSound(MusicToStop,0.0,option{Agent},false)

    Delay_NewQuestDisplay(Player : player)<suspends> : void =
        if(Agent := agent[Player], MainInventoryUI:=MaybeMainInventoryUI?):
            MainInventoryUI.Display_QuestCompleted()
            Sleep(5.0)
            if(Player.IsActive[]<#, PlayerData:=PlayersDataManager.GetPlayerData[Player]#>):
                set MainInventoryUI.IsQuestCompleted_BeingDisplayed = false
                MainInventoryUI.RefreshOrdersQuests(PlayerDataCache)

    Delay_NewOrderDisplay(Player : player)<suspends> : void =
        if(Agent := agent[Player], MainInventoryUI:=MaybeMainInventoryUI?):
            MainInventoryUI.Display_OrderCompleted()
            Sleep(5.0)
            if(Player.IsActive[]<#, PlayerData:=PlayersDataManager.GetPlayerData[Player]#>):
                set MainInventoryUI.IsOrderCompleted_BeingDisplayed = false
                MainInventoryUI.RefreshOrdersQuests(PlayerDataCache)

    AddProgressToOrder(Agent : agent) : void =
        if(Player := player[Agent], Player.IsActive[]<#, PlayerData := PlayersDataManager.GetPlayerData[Player], #>, MainInventoryUI:=MaybeMainInventoryUI?, ActionUI:=MaybeActionInfoUI?):
            #Amount calculated to add test
            AmountToAdd:=PlayerDataCache.OrderAssigned_ProductAmountRequested/3.0
            OrderInfoResult:=GameManager.QuestsAndOrdersDatabase.AddAgentOrder_Progress(Player,PlayerDataCache,AmountToAdd,AmountToAdd,AmountToAdd)
            if(OrderInfoResult(0)?):
                ##Print("OrderInfoResult(2) is {OrderInfoResult(2)}, OrderInfoResult(3) is {OrderInfoResult(3)}")
                ActionUI.DisplayGather(array{OrderInfoResult(2)}, array{OrderInfoResult(3)}, array{false},false)
                GameManager.VFXManager.BigObjetive_AudioPlayer.Play(Agent)
                spawn:
                    Delay_NewOrderDisplay(Player)
            else if(OrderInfoResult(1)?):
                ##Print("Refreshing order progress")
                MainInventoryUI.RefreshOrdersQuests(OrderInfoResult(4))
            
    Display_TeleporterMenu(Agent : agent, AreaNum : int) : void =
        if(AgentOwner:=MaybeAgentOwner?, Agent=AgentOwner):
            if(FastTravelUI:=MaybeFastTravelUI?):
                FastTravelUI.Display_Canvas(AreaNum)
        else:
            GameManager.DisplayCharacterMessage(Agent, I_CHARACTER, "This is not your teleporter!",?Sound:=ERROR_SOUND)

    Display_DropCoal(Agent : agent) : void =
        if(DropCoalUI:=MaybeDropCoalUI?):
            DropCoalUI.Display_Canvas()

    Display_ImproveDungeonTokens(Agent : agent) : void =
        if(TokenUI:=MaybeTokensUI?):
            TokenUI.DT_Display_Canvas()
        
    Display_ImproveHexaGameTokens(Agent : agent) : void =
        if(TokenUI:=MaybeTokensUI?):
            TokenUI.HT_Display_Canvas()

    Display_ImproveWeapons(WeaponType : int) : void =
        if(WeaponsUpgrades:=MaybeWeaponsUpgradesUI?):
            WeaponsUpgrades.Display_Canvas(WeaponType)

    Display_OptionsMenu() : void =
        if(OptionsMenu:=MaybeOptionsMenuUI?):
            OptionsMenu.Display_Canvas()
    
    Display_LanguageMenu() : void =
        if(LanguageMenu:=MaybeLanguageUI?):
            LanguageMenu.Display_Canvas()

            <#
    SetStatsFromAgentData() : void =
        #Set requirements based on player level
        if(AgentOwner := MaybeAgentOwner?, PlayerOwner := player[AgentOwner],AgentData := PlayersDataManager.GetPlayerData[PlayerOwner]:
            set LV_DroneSpeed = PlayerDataCache.LV_DroneSpeed
            set LV_DroneStorage = PlayerDataCache.LV_DroneStorage

            #set AutoWater_PerSecond = CalculateAutoWater_PerSecond(TLV_AutoWater#>

    ExecuteTickTime() : void =
        #Drone return
        if(TimeToDroneReturn = 7, not IsDronePropOnBase?):
            spawn:
                GameManager.VFXManager.AnimateDroneMovement(1, DroneProp, DroneInitialTransform, false)
        if(TimeToDroneReturn = 0, not IsDronePropOnBase?):
            DroneShip_Billbord.SetText(BasicMessage("READY FOR SHIPMENT"))
            set IsDronePropOnBase = true
        else if(TimeToDroneReturn > 0):
            set TimeToDroneReturn = TimeToDroneReturn - 1
            TupleMinutesSeconds:=TransformToMinutes_Seconds(TimeToDroneReturn)
            DroneShip_Billbord.SetText(BasicMessage("RETURNING IN:\n{TupleMinutesSeconds(0)}:{TupleMinutesSeconds(1)}"))
        DecrementMiniGamesCooldown()
        
    CheckQuest() : void =
        #Put this AFTER creating the new data of a new player and assigning to a bakery (MainInventoryUI)
        if(Agent:=MaybeAgentOwner?, Player:=player[Agent],Player.IsActive[]<#, PlayerData := PlayersDataManager.GetPlayerData[Player]#>, MainInventoryUI:=MaybeMainInventoryUI?):
            if(PlayerDataCache.NumQuestsCompleted=0, PlayerDataCache.QuestAssigned_NumQuest=-99):
                Log("Creating first quest",LOG_LEVEL_TUTORIALQUESTS)
                TupleQuest:=GameManager.QuestsAndOrdersDatabase.GetNewQuest(0,PlayerDataCache)
                Log("Quest {TupleQuest(0)} created and assigned",LOG_LEVEL_TUTORIALQUESTS)
                NewPlayerData := PlayersDataManager.CreateNewQuest(PlayerDataCache,0,TupleQuest(3),TupleQuest(5),TupleQuest(6))
                SetNewPlayerDataCache(NewPlayerData)
                #PlayersDataManager.SaveAllNewData(Player, PlayerData, NewPlayerData, Self)
                MainInventoryUI.RefreshAll(NewPlayerData, array{-1.0}, array{false})

    DisplayResourcersGather(Amounts : []float, Types : []int, IsCritical : logic) : void =
        ##Print("Trying to display Damage on screen A")
        var AllLimits : []logic = array{}
        for(A : Amounts):
            set AllLimits = AllLimits + array{false}
        for(Index:=0..Amounts.Length-1):
            if:
                AgentOwner:=MaybeAgentOwner?
                Amount:=Amounts[Index]
                Type:=Types[Index]
            then:
                if(Amount<0.0):
                    GameManager.VFXManager.PlayerPays_AudioPlayer.Play(AgentOwner)
                ##Print("Amounts: {Amount}")
                ##Print("Type: {Type}")
        if(ActionInfoUI:=MaybeActionInfoUI?):
            ##Print("Trying to display Damage on screen B")
            ActionInfoUI.DisplayGather(Amounts,Types,AllLimits,IsCritical)

    Delay_DisplayResourcersGather(Amounts : []float, Types : []int, IsCritical : logic, Delay : float)<suspends> : void =
        Sleep(Delay)
        DisplayResourcersGather(Amounts, Types, IsCritical)

    RefreshUKResourcers(AllResourcersType : []int, AllResourcersAmmount : []float) : void =
        if(MainInventoryUI:=MaybeMainInventoryUI?):
            MainInventoryUI.RefreshUKResourcers(AllResourcersType,AllResourcersAmmount)

    RefreshAllUI() : void =
        if:
            Agent:=MaybeAgentOwner?
            Player := player[Agent]
            Player.IsActive[]
            #PlayerData := PlayersDataManager.GetPlayerData[Player]
            MainInventoryUI:=MaybeMainInventoryUI?
            WaterLimit:=AllBackpackLimits[WATER]
            FlourLimit:=AllBackpackLimits[FLOUR]
            CheeseLimit:=AllBackpackLimits[CHEESE]
            PimentosLimit:=AllBackpackLimits[PIMENTOS]
        then:
            MainInventoryUI.RefreshAll(PlayerDataCache,array{WaterLimit,FlourLimit,CheeseLimit,PimentosLimit},array{false,false,false,false})

    RefreshInventoryUI(Agent : agent) : void =
        if:
            Player := player[Agent]
            Player.IsActive[]
            #PlayerData := PlayersDataManager.GetPlayerData[Player]
            MainInventoryUI:=MaybeMainInventoryUI?
            WaterLimit:=AllBackpackLimits[WATER]
            FlourLimit:=AllBackpackLimits[FLOUR]
            CheeseLimit:=AllBackpackLimits[CHEESE]
            PimentosLimit:=AllBackpackLimits[PIMENTOS]
        then:
            MainInventoryUI.RefreshInventory(PlayerDataCache,array{WaterLimit,FlourLimit,CheeseLimit,PimentosLimit},array{false,false,false,false})
        else:
            PrintFatalError("Fatal error trying to update UI")

    DisplayDungeonInfo() : void =
        CloseInputUIs()
        if(ActionInfoUI := MaybeActionInfoUI?):
            ActionInfoUI.SetDungeon_TimeInfo("Starting")
            ActionInfoUI.SetDungeon_ChestInfo("Prepare for battle")

    RemoveDungeonInfo() : void =
        if(ActionInfoUI:=MaybeActionInfoUI?):
            ActionInfoUI.RemoveDungeonInfo()

    UpdateDungeonUI(NumDungeon : int) : void =
        if(DungeonUI:=MaybeDungeonUI?,DungeonUI.Opened_NumDungeon=NumDungeon):
            DungeonUI.UpdateUI()

    Display_DungeonUI(Agent : agent, NumDungeon : int) : void =
        if(DungeonUI:=MaybeDungeonUI?):
            DungeonUI.Display_Canvas(NumDungeon)

    Display_MasteryUI(Agent : agent) : void =

        RequiredStrengthForMastery := Pow(10*1.0,MinExponential_ForAgedSapphires*1.0)
        if(PickaxeTotalStrength>Pow(10*1.0,MinExponential_ForAgedSapphires*1.0)*2.0):

        HardBreadLevelArea:=PlayerDataCache.LV_Area+1
        HardBreadHealth := GameManager.PickaxeCalculator.MainFormula_GenerateHardBreadHealth(HardBreadLevelArea,-1)

        ExpectedStrength:=HardBreadHealth/((MinNum_AVGHit+MaxNum_AVGHit)/2.0)
        AVGStrength:=(ExpectedStrength+PickaxeTotalStrength)/2.0
    
        if(RequiredStrengthForMastery<=AVGStrength):
            if(MasteryUI := MaybeMasteryUI?):
                MasteryUI.Display_Canvas()
        else if(AVGStrength>PickaxeTotalStrength):
            GameManager.DisplayCharacterMessage(Agent,I_CHARACTER,"You need {GetOptimizedMagnitudeLetter(RequiredStrengthForMastery-PickaxeTotalStrength)} more Pickaxe Strength!",?Sound:=ERROR_SOUND)  
        else:
            GameManager.DisplayCharacterMessage(Agent,I_CHARACTER,"You need to unlock more areas!",?Sound:=ERROR_SOUND)



    CloseInputUIs() : void =
        PrintWarning("TO DO Close all UI")
        if:
            DungeonInfoUI := MaybeDungeonUI?
        then:
            DungeonInfoUI.Remove_CanvasWithoutMessage()

    InsertResourcers(InAgent : agent): void =
        if(AgentOwner := MaybeAgentOwner?, InAgent = AgentOwner, Player := player[InAgent]<#, var PlayerData : player_data_table = PlayersDataManager.GetPlayerData[Player]#>):
            OldData := PlayerDataCache
            var AllResourcersInserted : []float = array{0.0,0.0,0.0,0.0}
            if:
                MainWaterDeposit_Limit := AllMainDeposit_Limits[WATER]
                MainFlourDeposit_Limit := AllMainDeposit_Limits[FLOUR]
                MainCheeseDeposit_Limit := AllMainDeposit_Limits[CHEESE]
                MainPimentosDeposit_Limit := AllMainDeposit_Limits[PIMENTOS]
            then:
                var AllResourcersLimited : []int = array{}
                #Water
                WaterLimitToInsert := MainWaterDeposit_Limit - PlayerDataCache.MainWaterDeposit_Value
                if(WaterLimitToInsert>=PlayerDataCache.Inventory_Water):
                    #set MainWaterDeposit_Value = MainWaterDeposit_Value + PlayerDataCache.Inventory_Water
                    SetNewPlayerDataCache(PlayersDataManager.SetWaterMainDeposit(PlayerDataCache,PlayerDataCache.MainWaterDeposit_Value + PlayerDataCache.Inventory_Water))
                    if(set AllResourcersInserted[WATER]= -PlayerDataCache.Inventory_Water):
                    SetNewPlayerDataCache(PlayersDataManager.WithdrawResource(PlayerDataCache,PlayerDataCache.Inventory_Water,WATER))
                else if(WaterLimitToInsert>0.0):
                    set AllResourcersLimited = AllResourcersLimited + array{WATER}
                    #set MainWaterDeposit_Value = MainWaterDeposit_Value + WaterLimitToInsert
                    SetNewPlayerDataCache(PlayersDataManager.SetWaterMainDeposit(PlayerDataCache,PlayerDataCache.MainWaterDeposit_Value + WaterLimitToInsert))
                    if(set AllResourcersInserted[WATER]= -WaterLimitToInsert):
                        SetNewPlayerDataCache(PlayersDataManager.WithdrawResource(PlayerDataCache,WaterLimitToInsert,WATER))
                #Flour
                FlourLimitToInsert:=MainFlourDeposit_Limit-PlayerDataCache.MainFlourDeposit_Value
                if(FlourLimitToInsert>=PlayerDataCache.Inventory_Flour):
                    #set MainFlourDeposit_Value=MainFlourDeposit_Value+PlayerDataCache.Inventory_Flour
                    SetNewPlayerDataCache(PlayersDataManager.SetFlourMainDeposit(PlayerDataCache,PlayerDataCache.MainFlourDeposit_Value + PlayerDataCache.Inventory_Flour))
                    if(set AllResourcersInserted[FLOUR]= -PlayerDataCache.Inventory_Flour):
                    SetNewPlayerDataCache(PlayersDataManager.WithdrawResource(PlayerDataCache,PlayerDataCache.Inventory_Flour,FLOUR))
                else if(FlourLimitToInsert>0.0):
                    set AllResourcersLimited = AllResourcersLimited + array{FLOUR}
                    #set MainFlourDeposit_Value=MainFlourDeposit_Value+FlourLimitToInsert
                    SetNewPlayerDataCache(PlayersDataManager.SetFlourMainDeposit(PlayerDataCache,PlayerDataCache.MainFlourDeposit_Value + FlourLimitToInsert))
                    if(set AllResourcersInserted[FLOUR]= -FlourLimitToInsert):
                    SetNewPlayerDataCache(PlayersDataManager.WithdrawResource(PlayerDataCache,FlourLimitToInsert,FLOUR))
                #Cheese
                CheeseLimitToInsert:=MainCheeseDeposit_Limit-PlayerDataCache.MainCheeseDeposit_Value
                if(CheeseLimitToInsert>=PlayerDataCache.Inventory_Cheese):
                    #set MainCheeseDeposit_Value=MainCheeseDeposit_Value+PlayerDataCache.Inventory_Cheese
                    SetNewPlayerDataCache(PlayersDataManager.SetCheeseMainDeposit(PlayerDataCache,PlayerDataCache.MainCheeseDeposit_Value + PlayerDataCache.Inventory_Cheese))
                    if(set AllResourcersInserted[CHEESE]= -PlayerDataCache.Inventory_Cheese):
                    SetNewPlayerDataCache(PlayersDataManager.WithdrawResource(PlayerDataCache,PlayerDataCache.Inventory_Cheese,CHEESE))
                else if(CheeseLimitToInsert>0.0):
                    set AllResourcersLimited = AllResourcersLimited + array{CHEESE}
                    #set MainCheeseDeposit_Value=MainCheeseDeposit_Value+CheeseLimitToInsert
                    SetNewPlayerDataCache(PlayersDataManager.SetCheeseMainDeposit(PlayerDataCache,PlayerDataCache.MainCheeseDeposit_Value + CheeseLimitToInsert))
                    if(set AllResourcersInserted[CHEESE]= -CheeseLimitToInsert):
                    SetNewPlayerDataCache(PlayersDataManager.WithdrawResource(PlayerDataCache,CheeseLimitToInsert,CHEESE))
                #Pimento
                PimentoLimitToInsert:=MainPimentosDeposit_Limit-PlayerDataCache.MainPimentosDeposit_Value
                if(PimentoLimitToInsert>=PlayerDataCache.Inventory_Pimentos):
                    #set MainPimentosDeposit_Value=MainPimentosDeposit_Value+PlayerDataCache.Inventory_Pimentos
                    SetNewPlayerDataCache(PlayersDataManager.SetPimentosMainDeposit(PlayerDataCache,PlayerDataCache.MainPimentosDeposit_Value+PlayerDataCache.Inventory_Pimentos))
                    if(set AllResourcersInserted[PIMENTOS]= -PlayerDataCache.Inventory_Pimentos):
                    SetNewPlayerDataCache(PlayersDataManager.WithdrawResource(PlayerDataCache,PlayerDataCache.Inventory_Pimentos,PIMENTOS))
                else if(PimentoLimitToInsert>0.0):
                    set AllResourcersLimited = AllResourcersLimited + array{PIMENTOS}
                    #set MainPimentosDeposit_Value=MainPimentosDeposit_Value+PimentoLimitToInsert
                    SetNewPlayerDataCache(PlayersDataManager.SetPimentosMainDeposit(PlayerDataCache,PlayerDataCache.MainPimentosDeposit_Value + PimentoLimitToInsert))
                    if(set AllResourcersInserted[PIMENTOS]= -PimentoLimitToInsert):
                    SetNewPlayerDataCache(PlayersDataManager.WithdrawResource(PlayerDataCache,PimentoLimitToInsert,PIMENTOS))
                
                if(WaterLimitToInsert > 0.0 or FlourLimitToInsert > 0.0 or CheeseLimitToInsert > 0.0 or PimentoLimitToInsert > 0.0):
                    GameManager.VFXManager.PlayerPays_AudioPlayer.Play(AgentOwner)
                    SetNewPlayerDataCache(GameManager.QuestsAndOrdersDatabase.AddAgentQuest_ProgressV2(Player,PlayerDataCache,1.0,USEFUNNEL_TUTORIAL_TYPE)(4))
                #MutantMaterial
                #set MainMutantMaterialDeposit_Value=MainMutantMaterialDeposit_Value+PlayerData.Inventory_MutantMaterial
                #set PlayerData = PlayersDataManager.WithdrawResource(PlayerData,PlayerData.Inventory_MutantMaterial,MUTANTMATERIAL)
                #Saving new data!
                #PlayersDataManager.SaveAllNewData(Player, OldData, PlayerData, Self)
                DisplayResourcersGather(AllResourcersInserted,array{WATER,FLOUR,CHEESE,PIMENTOS},false)
                UpdateAllMainDeposits()
                RefreshInventoryUI(PlayerDataCache)
                #Limited message
                if(AllResourcersLimited.Length>0 or (PlayerDataCache.MainWaterDeposit_Value=MainWaterDeposit_Limit) or (PlayerDataCache.MainFlourDeposit_Value=MainFlourDeposit_Limit) or (PlayerDataCache.MainCheeseDeposit_Value=MainCheeseDeposit_Limit) or (PlayerDataCache.MainPimentosDeposit_Value=MainPimentosDeposit_Limit)):
                    var StringResourcers : string = ""
                    for(Type : AllResourcersLimited):
                        if(Type = WATER):
                            set StringResourcers = StringResourcers + "WATER "
                        else if(Type = FLOUR):
                            set StringResourcers = StringResourcers + "FLOUR "
                        else if(Type = CHEESE): 
                            set StringResourcers = StringResourcers + "CHEESE "
                        else if(Type = PIMENTOS): 
                            set StringResourcers = StringResourcers + "PIMENTOS "
                    var StringMessage : string = "Storage full! Upgrade capacity recommended"
                    GameManager.DisplayCharacterMessage(AgentOwner,I_CHARACTER,StringMessage,?Sound:=FULL_SOUND)
        else:
            GameManager.Display_NotOwner(InAgent)

    FillPasiveMainDeposits() : void =
        if:
            MainWaterDeposit_Limit := AllMainDeposit_Limits[WATER]
            MainFlourDeposit_Limit := AllMainDeposit_Limits[FLOUR]
            MainCheeseDeposit_Limit := AllMainDeposit_Limits[CHEESE]
            MainPimentosDeposit_Limit := AllMainDeposit_Limits[PIMENTOS]
            PasiveWaterIncome := AllPasiveResourcers[WATER]
            PasiveFlourIncome := AllPasiveResourcers[FLOUR]
            PasiveCheeseIncome := AllPasiveResourcers[CHEESE]
            PasivePimentosIncome := AllPasiveResourcers[PIMENTOS]

        then:
            var WaterDeposit_Value : float = PasiveWaterIncome + PlayerDataCache.MainWaterDeposit_Value
            var FlourDeposit_Value : float = PasiveFlourIncome + PlayerDataCache.MainFlourDeposit_Value
            var CheeseDeposit_Value: float = PasiveCheeseIncome + PlayerDataCache.MainCheeseDeposit_Value
            var PimentosDeposit_Value : float = PasivePimentosIncome + PlayerDataCache.MainPimentosDeposit_Value

            if(WaterDeposit_Value > MainWaterDeposit_Limit):
                set WaterDeposit_Value = MainWaterDeposit_Limit
            if(FlourDeposit_Value > MainFlourDeposit_Limit):
                set FlourDeposit_Value = MainFlourDeposit_Limit
            if(CheeseDeposit_Value > MainCheeseDeposit_Limit):
                set CheeseDeposit_Value = MainCheeseDeposit_Limit
            if(PimentosDeposit_Value > MainPimentosDeposit_Limit):
                set PimentosDeposit_Value = MainPimentosDeposit_Limit

            SetNewPlayerDataCache(PlayersDataManager.SetMainDeposits(PlayerDataCache,WaterDeposit_Value,FlourDeposit_Value,CheeseDeposit_Value,PimentosDeposit_Value))
            #set MainWaterDeposit_Value = PasiveWaterIncome + PlayerDataCache.MainWaterDeposit_Value
            #set MainFlourDeposit_Value = PasiveFlourIncome + MainFlourDeposit_Value
            #set MainCheeseDeposit_Value = PasiveCheeseIncome + MainCheeseDeposit_Value
            #set MainPimentosDeposit_Value = PasivePimentosIncome + MainPimentosDeposit_Value

            

        UpdateAllMainDeposits()

    TakeWater() : void =
        var NumLoops : int = 0
        set WaterFlowPerSecond_ForB_ToTake = 0.0
        set WaterFlowPerSecond_ForCB_ToTake = 0.0
        set WaterFlowPerSecond_ForPCS_ToTake = 0.0
        loop:
            set NumLoops = NumLoops + 1
            var Total_WaterValves_Values : float = 0.0
                #Checking total water values that are open
            if(PlayerDataCache.WaterPreparedFor_B < Req_WaterFor_B):
                set Total_WaterValves_Values = Total_WaterValves_Values + ValveState_WaterFor_B
            if(PlayerDataCache.WaterPreparedFor_CB < Req_WaterFor_CB):
                set Total_WaterValves_Values = Total_WaterValves_Values + ValveState_WaterFor_CB
            if(PlayerDataCache.WaterPreparedFor_PCS < Req_WaterFor_PCS):
                set Total_WaterValves_Values = Total_WaterValves_Values + ValveState_WaterFor_PCS
                #There is nothing to distribute
            if(Total_WaterValves_Values <= 0.0):
                #Log("All the deposits are full of the valves are closed", LOG_LEVEL_BAKERS)
                set WaterAmmount_ForB_ToTake = 0.0
                set WaterAmmount_ForCB_ToTake = 0.0
                set WaterAmmount_ForPCS_ToTake = 0.0
                break
            else:
                #There is something to distribute
                set WaterAmmount_ForB_ToTake=Calculate_ResTaken(PlayerDataCache.MainWaterDeposit_Value, ValveState_WaterFor_B, Total_WaterValves_Values, Req_WaterFor_B, PlayerDataCache.WaterPreparedFor_B)
                set WaterAmmount_ForCB_ToTake=Calculate_ResTaken(PlayerDataCache.MainWaterDeposit_Value, ValveState_WaterFor_CB, Total_WaterValves_Values, Req_WaterFor_CB, PlayerDataCache.WaterPreparedFor_CB)
                set WaterAmmount_ForPCS_ToTake=Calculate_ResTaken(PlayerDataCache.MainWaterDeposit_Value, ValveState_WaterFor_PCS, Total_WaterValves_Values, Req_WaterFor_PCS, PlayerDataCache.WaterPreparedFor_PCS)
                    #Acumulate resorce to flow per second
                set WaterFlowPerSecond_ForB_ToTake = WaterFlowPerSecond_ForB_ToTake + WaterAmmount_ForB_ToTake
                set WaterFlowPerSecond_ForCB_ToTake = WaterFlowPerSecond_ForCB_ToTake + WaterAmmount_ForCB_ToTake
                set WaterFlowPerSecond_ForPCS_ToTake = WaterFlowPerSecond_ForPCS_ToTake + WaterAmmount_ForPCS_ToTake
                    #Removing resource from deposit
                SetNewPlayerDataCache(PlayersDataManager.SetWaterMainDeposit(PlayerDataCache, PlayerDataCache.MainWaterDeposit_Value - WaterAmmount_ForB_ToTake - WaterAmmount_ForCB_ToTake - WaterAmmount_ForPCS_ToTake))
                    #Distributing the water water
                SetNewPlayerDataCache(PlayersDataManager.SetWaterSubDeposit(PlayerDataCache,PlayerDataCache.WaterPreparedFor_B+WaterAmmount_ForB_ToTake,PlayerDataCache.WaterPreparedFor_CB+WaterAmmount_ForCB_ToTake,PlayerDataCache.WaterPreparedFor_PCS+WaterAmmount_ForPCS_ToTake))
                    #If there is no water available, break the loop
                if(PlayerDataCache.MainWaterDeposit_Value < 1.0):
                    #Log("Remaining water: {MainWaterDeposit_Value}", LOG_LEVEL_BAKERS) 
                    break
                if(NumLoops>MaxLoops):
                    PrintFatalError("FATAL ERROR infinite loop on TakeWater()")
                    break
    TakeFlour() : void =
            var NumLoops : int = 0
            set FlourFlowPerSecond_ForB_ToTake = 0.0
            set FlourFlowPerSecond_ForCB_ToTake = 0.0
            set FlourFlowPerSecond_ForPCS_ToTake = 0.0
            loop:
                set NumLoops = NumLoops + 1
                var Total_FlourValves_Values : float = 0.0
                    #Checking total flour values that are open
                if(PlayerDataCache.FlourPreparedFor_B < Req_FlourFor_B):
                    set Total_FlourValves_Values = Total_FlourValves_Values + ValveState_FlourFor_B
                if(PlayerDataCache.FlourPreparedFor_CB < Req_FlourFor_CB):
                    set Total_FlourValves_Values = Total_FlourValves_Values + ValveState_FlourFor_CB
                if(PlayerDataCache.FlourPreparedFor_PCS < Req_FlourFor_PCS):
                    set Total_FlourValves_Values = Total_FlourValves_Values + ValveState_FlourFor_PCS
                    #There is nothing to distribute
                if(Total_FlourValves_Values <= 0.0):
                    #Log("All the deposits are full of the valves are closed", LOG_LEVEL_BAKERS)
                    set FlourAmmount_ForB_ToTake = 0.0
                    set FlourAmmount_ForCB_ToTake = 0.0
                    set FlourAmmount_ForPCS_ToTake = 0.0
                    break
                else:
                    #There is something to distribute
                    set FlourAmmount_ForB_ToTake=Calculate_ResTaken(PlayerDataCache.MainFlourDeposit_Value, ValveState_FlourFor_B, Total_FlourValves_Values, Req_FlourFor_B, PlayerDataCache.FlourPreparedFor_B)
                    set FlourAmmount_ForCB_ToTake=Calculate_ResTaken(PlayerDataCache.MainFlourDeposit_Value, ValveState_FlourFor_CB, Total_FlourValves_Values, Req_FlourFor_CB, PlayerDataCache.FlourPreparedFor_CB)
                    set FlourAmmount_ForPCS_ToTake=Calculate_ResTaken(PlayerDataCache.MainFlourDeposit_Value, ValveState_FlourFor_PCS, Total_FlourValves_Values, Req_FlourFor_PCS, PlayerDataCache.FlourPreparedFor_PCS)
                        #Removing resource from deposit
                    SetNewPlayerDataCache(PlayersDataManager.SetFlourMainDeposit(PlayerDataCache, PlayerDataCache.MainFlourDeposit_Value - FlourAmmount_ForB_ToTake - FlourAmmount_ForCB_ToTake - FlourAmmount_ForPCS_ToTake))
                        #Acumulate resorce to flow per second
                    set FlourFlowPerSecond_ForB_ToTake = FlourFlowPerSecond_ForB_ToTake + FlourAmmount_ForB_ToTake
                    set FlourFlowPerSecond_ForCB_ToTake = FlourFlowPerSecond_ForCB_ToTake + FlourAmmount_ForCB_ToTake
                    set FlourFlowPerSecond_ForPCS_ToTake = FlourFlowPerSecond_ForPCS_ToTake + FlourAmmount_ForPCS_ToTake
                        #Distributing the flour
                    SetNewPlayerDataCache(PlayersDataManager.SetFlourSubDeposit(PlayerDataCache,PlayerDataCache.FlourPreparedFor_B+FlourAmmount_ForB_ToTake,PlayerDataCache.FlourPreparedFor_CB+FlourAmmount_ForCB_ToTake,PlayerDataCache.FlourPreparedFor_PCS+FlourAmmount_ForPCS_ToTake))  
                    #Log("FlourAmmount_ForB_ToTake flour: {FlourAmmount_ForB_ToTake}", LOG_LEVEL_BAKERS)
                    #Log("FlourAmmount_ForCB_ToTake flour: {FlourAmmount_ForCB_ToTake}", LOG_LEVEL_BAKERS)
                    #Log("FlourAmmount_ForPCS_ToTake flour: {FlourAmmount_ForPCS_ToTake}", LOG_LEVEL_BAKERS)
                        #If there is no flour available, break the loop
                    if(PlayerDataCache.MainFlourDeposit_Value < 1.0):
                        #Log("Remaining flour: {MainFlourDeposit_Value}", LOG_LEVEL_BAKERS) 
                        break
                    if(NumLoops>MaxLoops):
                        PrintFatalError("FATAL ERROR infinite loop on TakeFlour()")
                        break
    TakeCheese() : void =
        var NumLoops : int = 0
        set CheeseFlowPerSecond_ForCB_ToTake = 0.0
        set CheeseFlowPerSecond_ForPCS_ToTake = 0.0
        loop:
            set NumLoops = NumLoops + 1
            var Total_CheeseValves_Values : float = 0.0
                #Checking total cheese values that are open
            if(PlayerDataCache.CheesePreparedFor_CB < Req_CheeseFor_CB):
                set Total_CheeseValves_Values = Total_CheeseValves_Values + ValveState_CheeseFor_CB
            if(PlayerDataCache.CheesePreparedFor_PCS < Req_CheeseFor_PCS):
                set Total_CheeseValves_Values = Total_CheeseValves_Values + ValveState_CheeseFor_PCS
                #There is nothing to distribute
            if(Total_CheeseValves_Values <= 0.0):
                #Log("All the cheese deposits are full of the cheese valves are closed", LOG_LEVEL_BAKERS)
                set CheeseAmmount_ForCB_ToTake = 0.0
                set CheeseAmmount_ForPCS_ToTake = 0.0
                break
            else:
                #There is something to distribute
                set CheeseAmmount_ForCB_ToTake=Calculate_ResTaken(PlayerDataCache.MainCheeseDeposit_Value, ValveState_CheeseFor_CB, Total_CheeseValves_Values, Req_CheeseFor_CB, PlayerDataCache.CheesePreparedFor_CB)
                set CheeseAmmount_ForPCS_ToTake=Calculate_ResTaken(PlayerDataCache.MainCheeseDeposit_Value, ValveState_CheeseFor_PCS, Total_CheeseValves_Values, Req_CheeseFor_PCS, PlayerDataCache.CheesePreparedFor_PCS)
                    #Removing resource from deposit
                SetNewPlayerDataCache(PlayersDataManager.SetCheeseMainDeposit(PlayerDataCache, PlayerDataCache.MainCheeseDeposit_Value - CheeseAmmount_ForCB_ToTake - CheeseAmmount_ForPCS_ToTake))
                    #Acumulate resorce to flow per second
                set CheeseFlowPerSecond_ForCB_ToTake = CheeseFlowPerSecond_ForCB_ToTake + CheeseAmmount_ForCB_ToTake
                set CheeseFlowPerSecond_ForPCS_ToTake = CheeseFlowPerSecond_ForPCS_ToTake + CheeseAmmount_ForPCS_ToTake   
                    #Distributing the cheese
                SetNewPlayerDataCache(PlayersDataManager.SetCheeseSubDeposit(PlayerDataCache,PlayerDataCache.CheesePreparedFor_CB+CheeseAmmount_ForCB_ToTake,PlayerDataCache.CheesePreparedFor_PCS+CheeseAmmount_ForPCS_ToTake))  
                #Log("CheesePreparedFor_CB: {CheesePreparedFor_CB}", LOG_LEVEL_BAKERS)
                #Log("CheeseAmmount_ForCB_ToTake cheese: {CheeseAmmount_ForCB_ToTake}", LOG_LEVEL_BAKERS)
                #Log("CheeseAmmount_ForPCS_ToTake cheese: {CheeseAmmount_ForPCS_ToTake}", LOG_LEVEL_BAKERS)
                    #If there is no cheese available, break the loop
                if(PlayerDataCache.MainCheeseDeposit_Value < 1.0):
                    #Log("Remaining cheese: {MainCheeseDeposit_Value}", LOG_LEVEL_BAKERS) 
                    break
                if(NumLoops>MaxLoops):
                    PrintFatalError("FATAL ERROR infinite loop on TakeCheese()")
                    break
        
    TakePimentos() : void =
        var NumLoops : int = 0
        #set PimentosFlowPerSecond_ForPCS_ToTake = 0.0
        #set PimentosFlowPerSecond_ForPets_ToTake = 0.0
        set ValveState_PimentosFor_PCS = 1.0
        loop:
            set NumLoops = NumLoops + 1
            var Total_PimentosValves_Values : float = 0.0
                #Checking total cheese values that are open
            if(PlayerDataCache.PimentosPreparedFor_PCS < Req_PimentosFor_PCS):
                set Total_PimentosValves_Values = Total_PimentosValves_Values + ValveState_PimentosFor_PCS
            #if(PimentosPreparedFor_Pets < Max_PimentosFor_Pets, Max_PimentosFor_Pets > 0.0):
            #    set Total_PimentosValves_Values = 0.0
                #There is nothing to distribute
            if(Total_PimentosValves_Values <= 0.0):
                #Log("All the cheese deposits are full of the cheese valves are closed", LOG_LEVEL_BAKERS)
                set PimentosAmmount_ForPCS_ToTake = 0.0
                #set PimentosAmmount_ForPets_ToTake = 0.0
                break
            else:
                #There is something to distribute
                set PimentosAmmount_ForPCS_ToTake=Calculate_ResTaken(PlayerDataCache.MainPimentosDeposit_Value, ValveState_PimentosFor_PCS, Total_PimentosValves_Values, Req_PimentosFor_PCS, PlayerDataCache.PimentosPreparedFor_PCS)
                #set PimentosAmmount_ForPets_ToTake=Calculate_ResTaken(MainPimentosDeposit_Value, ValveState_PimentosFor_Pets, Total_PimentosValves_Values, Max_PimentosFor_Pets, PimentosPreparedFor_Pets)
                    #Removing resource from deposit
                SetNewPlayerDataCache(PlayersDataManager.SetPimentosMainDeposit(PlayerDataCache, PlayerDataCache.MainPimentosDeposit_Value - PimentosAmmount_ForPCS_ToTake))
                    #Acumulate resorce to flow per second
                set PimentosFlowPerSecond_ForPCS_ToTake = PimentosFlowPerSecond_ForPCS_ToTake + PimentosAmmount_ForPCS_ToTake
                #set PimentosFlowPerSecond_ForPets_ToTake = PimentosFlowPerSecond_ForPets_ToTake + PimentosAmmount_ForPets_ToTake   
                    #Distributing the pimentos
                SetNewPlayerDataCache(PlayersDataManager.SetPimentoSubDeposit(PlayerDataCache,PlayerDataCache.PimentosPreparedFor_PCS + PimentosAmmount_ForPCS_ToTake))     
                #set PimentosPreparedFor_Pets = PimentosPreparedFor_Pets + PimentosAmmount_ForPets_ToTake
                #Log("CheesePreparedFor_CB: {CheesePreparedFor_CB}", LOG_LEVEL_BAKERS)
                #Log("CheeseAmmount_ForCB_ToTake cheese: {CheeseAmmount_ForCB_ToTake}", LOG_LEVEL_BAKERS)
                #Log("CheeseAmmount_ForPCS_ToTake cheese: {CheeseAmmount_ForPCS_ToTake}", LOG_LEVEL_BAKERS)
                    #If there is no cheese available, break the loop
                if(PlayerDataCache.MainPimentosDeposit_Value < 1.0):
                    #Log("Remaining cheese: {MainCheeseDeposit_Value}", LOG_LEVEL_BAKERS) 
                    break
                if(NumLoops>MaxLoops):
                    PrintFatalError("FATAL ERROR infinite loop on TakeCheese()")
                    break

    UpgradeResourcersContainersUI() : void =

        #At the final of pimentos, upgrade all UI that upgrade the amount of recourcers available
        if(BackpackUpgradesUI := MaybeBackpackUpgradesUI?, BackpackUpgradesUI.UIIsDisplayed?):
            BackpackUpgradesUI.UpdateUI(PlayerDataCache)
        if(DepositsUpgradeUI := MaybeDepositsUpgradeUI?, DepositsUpgradeUI.UIIsDisplayed?):
            DepositsUpgradeUI.UpdateUI(PlayerDataCache)


    
    <#
    TakePimentos() : void =
        var NumLoops : int = 0
        #set PimentosFlowPerSecond_ForPCS_ToTake = 0.0
        #set PimentosFlowPerSecond_ForPets_ToTake = 0.0
        loop:
            set NumLoops = NumLoops + 1
            var Total_PimentosValves_Values : float = 0.0
                #Checking total cheese values that are open
            if(PimentosPreparedFor_PCS < Req_PimentosFor_PCS):
                set Total_PimentosValves_Values = Total_PimentosValves_Values + ValveState_PimentosFor_PCS
            if(PimentosPreparedFor_Pets < Max_PimentosFor_Pets, Max_PimentosFor_Pets > 0.0):
                set Total_PimentosValves_Values = Total_PimentosValves_Values + ValveState_PimentosFor_Pets
                #There is nothing to distribute
            if(Total_PimentosValves_Values <= 0.0):
                #Log("All the cheese deposits are full of the cheese valves are closed", LOG_LEVEL_BAKERS)
                set PimentosAmmount_ForPCS_ToTake = 0.0
                set PimentosAmmount_ForPets_ToTake = 0.0
                break
            else:
                #There is something to distribute
                set PimentosAmmount_ForPCS_ToTake=Calculate_ResTaken(MainPimentosDeposit_Value, ValveState_PimentosFor_PCS, Total_PimentosValves_Values, Req_PimentosFor_PCS, PimentosPreparedFor_PCS)
                #set PimentosAmmount_ForPets_ToTake=Calculate_ResTaken(MainPimentosDeposit_Value, ValveState_PimentosFor_Pets, Total_PimentosValves_Values, Max_PimentosFor_Pets, PimentosPreparedFor_Pets)
                    #Removing resource from deposit
                set MainPimentosDeposit_Value = MainPimentosDeposit_Value - PimentosAmmount_ForPCS_ToTake - PimentosAmmount_ForPets_ToTake
                    #Acumulate resorce to flow per second
                set PimentosFlowPerSecond_ForPCS_ToTake = PimentosFlowPerSecond_ForPCS_ToTake + PimentosAmmount_ForPCS_ToTake
                set PimentosFlowPerSecond_ForPets_ToTake = PimentosFlowPerSecond_ForPets_ToTake + PimentosAmmount_ForPets_ToTake   
                    #Distributing the cheese
                set PimentosPreparedFor_PCS = PimentosPreparedFor_PCS + PimentosAmmount_ForPCS_ToTake
                set PimentosPreparedFor_Pets = PimentosPreparedFor_Pets + PimentosAmmount_ForPets_ToTake
                #Log("CheesePreparedFor_CB: {CheesePreparedFor_CB}", LOG_LEVEL_BAKERS)
                #Log("CheeseAmmount_ForCB_ToTake cheese: {CheeseAmmount_ForCB_ToTake}", LOG_LEVEL_BAKERS)
                #Log("CheeseAmmount_ForPCS_ToTake cheese: {CheeseAmmount_ForPCS_ToTake}", LOG_LEVEL_BAKERS)
                    #If there is no cheese available, break the loop
                if(MainPimentosDeposit_Value < 1.0):
                    #Log("Remaining cheese: {MainCheeseDeposit_Value}", LOG_LEVEL_BAKERS) 
                    break
                if(NumLoops>MaxLoops):
                    PrintFatalError("FATAL ERROR infinite loop on TakeCheese()")
                    break#>
    #PRODUCTION
    #Super energy
    TryConsumeSuperEnergy() : void =
        #OwnedEnergy
        if(Agent:=MaybeAgentOwner?, Player := player[Agent]<#, PlayerData:=GameManager.PlayersDataManager.GetPlayerData[Player]#>):

            BakerIndex:=GameManager.GetAgentIndex(Agent)
            var AmountSuperEnergy_Multiplier : float = 1.0
            #Shared Energy
            IsThereOtherSharedEnergy := GameManager.IsThereSharedSuperEnergy(BakerIndex)
            if(IsThereOtherSharedEnergy?):
                if(PlayerDataCache.SuperEnergy_Owned>0.0):
                    set AmountSuperEnergy_Multiplier = SuperEnergyMultiplier_Owned + SuperEnergyMultiplier_Shared
                    #Log("Is there Shared Energy, and the Player {BakerIndex} Has Super Energy. Multiplier: {AmountSuperEnergy_Multiplier}", LOG_LEVEL_DEBUG9) 
                else:
                    set AmountSuperEnergy_Multiplier = 1.0 + SuperEnergyMultiplier_Shared
                    #Log("Is there Shared Energy, and the Player {BakerIndex} Has Not Super Energy. Multiplier: {AmountSuperEnergy_Multiplier}", LOG_LEVEL_DEBUG9)
            else:
                if(PlayerDataCache.SuperEnergy_Owned>0.0):
                    set AmountSuperEnergy_Multiplier = SuperEnergyMultiplier_Owned
                    #Log("There is NOT Shared Energy, the Player {BakerIndex} Has Super Energy. Multiplier: {SuperEnergyMultiplier_Owned}", LOG_LEVEL_DEBUG9)
                else:
                    set AmountSuperEnergy_Multiplier = 1.0
                    #Log("There is NOT Shared Energy, the Player {BakerIndex} Has not Super Energy. Multiplier: {AmountSuperEnergy_Multiplier}", LOG_LEVEL_DEBUG9)

            var NewSuperEnergy_Owned : float = PlayerDataCache.SuperEnergy_Owned - 1.0 #Consume
            #Log("SuperEnergy_Owned: {SuperEnergy_Owned}", LOG_LEVEL_COAL) 
            if(NewSuperEnergy_Owned < 0.0):
                set NewSuperEnergy_Owned = 0.0
            if(NewSuperEnergy_Owned > 0.0):
                #Super energy props
                if:
                    BB_SuperEneryProp:=MaybeBB_SuperEneryProp?
                    BB_SuperEneryProp.IsValid[]
                    CBB_SuperEneryProp:=MaybeCBB_SuperEneryProp?
                    CBB_SuperEneryProp.IsValid[]
                    PCSB_SuperEneryProp:=MaybePCSB_SuperEneryProp?
                    PCSB_SuperEneryProp.IsValid[]
                else:
                    #Log("Spawning SuperEnergyProps", LOG_LEVEL_COAL)
                    SpawnTuple_SuperEnergyBB:=SpawnProp(GameManager.VFXManager.BakerGeneratorProp_SuperEnergy,BB_Prop.GetTransform())
                    SpawnTuple_SuperEnergyCBB:=SpawnProp(GameManager.VFXManager.BakerGeneratorProp_SuperEnergy,CBB_Prop.GetTransform())
                    SpawnTuple_SuperEnergyPCSB:=SpawnProp(GameManager.VFXManager.BakerGeneratorProp_SuperEnergy,PCSB_Prop.GetTransform())
                    set MaybeBB_SuperEneryProp = SpawnTuple_SuperEnergyBB(0)
                    set MaybeCBB_SuperEneryProp = SpawnTuple_SuperEnergyCBB(0)
                    set MaybePCSB_SuperEneryProp = SpawnTuple_SuperEnergyPCSB(0)
            else:
                #Log("Trying to remove SuperEnergyProps because SuperEnergy_Owned: {SuperEnergy_Owned}", LOG_LEVEL_COAL)
                #Remove props if they are valid
                if:
                    BB_SuperEneryProp:=MaybeBB_SuperEneryProp?
                    BB_SuperEneryProp.IsValid[]
                    CBB_SuperEneryProp:=MaybeCBB_SuperEneryProp?
                    CBB_SuperEneryProp.IsValid[]
                    PCSB_SuperEneryProp:=MaybePCSB_SuperEneryProp?
                    PCSB_SuperEneryProp.IsValid[]
                then:
                    #Log("Props disposed", LOG_LEVEL_COAL)
                    BB_SuperEneryProp.Dispose()
                    CBB_SuperEneryProp.Dispose()
                    PCSB_SuperEneryProp.Dispose()
                    set MaybeBB_SuperEneryProp = false
                    set MaybeCBB_SuperEneryProp = false
                    set MaybePCSB_SuperEneryProp = false
            
            #Set Timer  
            if(NewSuperEnergy_Owned <= 0.0):
                SuperEnergyRemaining_Timer.SetActiveDuration(0.0)
                set NewSuperEnergy_Owned = 0.0
            else:
                SuperEnergyRemaining_Timer.SetMaxDuration(NewSuperEnergy_Owned) 
            SetNewPlayerDataCache(PlayersDataManager.SetSuperEnergy(PlayerDataCache,NewSuperEnergy_Owned))
            #Save data
            #NewPlayerData:=GameManager.PlayersDataManager.SetSuperEnergy(PlayerData,SuperEnergy_Owned)
            #PlayersDataManager.SaveNewData(Player,PlayerData,NewPlayerData) 
            set SuperEnergy_Multiplier = AmountSuperEnergy_Multiplier
        SetSuperEnergyMultiplierProp()

    <#
    TryConsumeSuperEnergy(Amount : float, IsProduction : logic) : float =
        var OutPutAmount : float = Amount
        var ProductionMultiplier : float = 1.0
        #Add from shared energy
        IsThereSharedEnergy := GameManager.IsThereSharedSuperEnergy()
        if(IsThereSharedEnergy?):
            if(IsProduction?):
                set ProductionMultiplier = ProductionMultiplier + SuperEnergyProductionMultiplier_Shared
                #set OutPutAmount = OutPutAmount + Amount * SuperEnergyProductionMultiplier_Shared
            else:
                set OutPutAmount = OutPutAmount + Amount * SuperEnergySpeedMultiplier_Shared
        if(Agent :=MaybeAgentOwner?, Player := player[Agent], PlayerData:=GameManager.PlayersDataManager.GetPlayerData[Player]):
            var SuperEnergy_Owned : float = PlayerData.SuperEnergy_Owned
            #Add from owner energy
            if(PlayerData.SuperEnergy_Owned > 0.0):
                set SuperEnergy_Owned = SuperEnergy_Owned - 1.0
                SuperEnergyRemaining_Timer.SetMaxDuration(SuperEnergy_Owned*1.0)
                if(IsProduction?):
                    set ProductionMultiplier = ProductionMultiplier + SuperEnergyProductionMultiplier_Shared
                    #set OutPutAmount = OutPutAmount + Amount * SuperEnergyProductionMultiplier
                else:
                    set OutPutAmount = OutPutAmount + Amount * SuperEnergySpeedMultiplier
            #Super energy props
                if:
                    BB_SuperEneryProp:=MaybeBB_SuperEneryProp?
                    BB_SuperEneryProp.IsValid[]
                    CBB_SuperEneryProp:=MaybeCBB_SuperEneryProp?
                    CBB_SuperEneryProp.IsValid[]
                    PCSB_SuperEneryProp:=MaybePCSB_SuperEneryProp?
                    PCSB_SuperEneryProp.IsValid[]
                else:
                    SpawnTuple_SuperEnergyBB:=SpawnProp(GameManager.VFXManager.BakerGeneratorProp_SuperEnergy,BB_Prop.GetTransform())(0)
                    SpawnTuple_SuperEnergyCBB:=SpawnProp(GameManager.VFXManager.BakerGeneratorProp_SuperEnergy,CBB_Prop.GetTransform())(0)
                    SpawnTuple_SuperEnergyPCSB:=SpawnProp(GameManager.VFXManager.BakerGeneratorProp_SuperEnergy,PCSB_Prop.GetTransform())(0)
                    set MaybeBB_SuperEneryProp = SpawnTuple_SuperEnergyBB
                    set MaybeCBB_SuperEneryProp = SpawnTuple_SuperEnergyCBB
                    set MaybePCSB_SuperEneryProp = SpawnTuple_SuperEnergyPCSB
            else:
                if:
                    BB_SuperEneryProp:=MaybeBB_SuperEneryProp?
                    BB_SuperEneryProp.IsValid[]
                    CBB_SuperEneryProp:=MaybeCBB_SuperEneryProp?
                    CBB_SuperEneryProp.IsValid[]
                    PCSB_SuperEneryProp:=MaybePCSB_SuperEneryProp?
                    PCSB_SuperEneryProp.IsValid[]
                then:
                    BB_SuperEneryProp.Dispose()
                    CBB_SuperEneryProp.Dispose()
                    PCSB_SuperEneryProp.Dispose()
                    set MaybeBB_SuperEneryProp = false
                    set MaybeCBB_SuperEneryProp = false
                    set MaybePCSB_SuperEneryProp = false

            #Set timer stat
            if(SuperEnergy_Owned < 0.0):
                set SuperEnergy_Owned = 0.0
            #Log("SuperEnergy_Shared: {SuperEnergy_Owned}", LOG_LEVEL_1DETAILING)  
            if(SuperEnergy_Owned<=0.0):
                SuperEnergyRemaining_Timer.SetActiveDuration(0.0)
            else:
                SuperEnergyRemaining_Timer.SetMaxDuration(SuperEnergy_Owned) 

            NewPlayerData:=GameManager.PlayersDataManager.SetSuperEnergy(PlayerData,SuperEnergy_Owned)
            PlayersDataManager.SaveNewData(Player,PlayerData,NewPlayerData)  

            

        return OutPutAmount #>

    SetSuperEnergyMultiplierProp<private>() : void =
        #Log("SuperEnergy_Multiplier: {SuperEnergy_Multiplier}", LOG_LEVEL_COAL)
        if:
            Unidades := Floor[SuperEnergy_Multiplier/1.0]
            Decimal := Floor[(SuperEnergy_Multiplier-Unidades*1.0)/0.1]
        then:
            Remaining := SuperEnergy_Multiplier - (Unidades*1.0) - (Decimal*0.1)
            #Log("Remaining: {Remaining}", LOG_LEVEL_COAL)
            #Centesimal := Floor[(SuperEnergy_Multiplier-(Unidades*1.0)-(Decimal*0.1))/0.01]
            CentesimalFloat := Remaining/0.01
            #Log("CentesimalFloat: {CentesimalFloat}", LOG_LEVEL_COAL)

            if(CentesimalInt := Ceil[CentesimalFloat]):
                #Log("Unidades: {Unidades}", LOG_LEVEL_COAL)
                #Log("Decimal: {Decimal}", LOG_LEVEL_COAL)
                #Log("Centesimal: {CentesimalInt}", LOG_LEVEL_COAL)

                if:
                    UnidadesMaterial:=AllMaterialnumbers[Unidades]
                    DecimalMaterial:=AllMaterialnumbers[Decimal]
                    CentesimalMaterial:=AllMaterialnumbers[CentesimalInt]
                then:
                    OverPoweredSign.SetMaterial(UnidadesMaterial,?Index:=2)
                    OverPoweredSign.SetMaterial(DecimalMaterial,?Index:=3)      
                    OverPoweredSign.SetMaterial(CentesimalMaterial,?Index:=4)  
        else:
            PrintFatalError("Fatal error on SetSuperEnergyMultiplierProp")


    SetMaterialCost<private>(CreativeProp : creative_prop, Button : button_device, AmountCost : float) : void =
        if:
            Decenas := Floor[AmountCost/10.0]
            Unidades := Floor[AmountCost-(Decenas*10.0)]
            IntAmount := Floor[AmountCost]
        then:
            if(DecenasMaterial:=AllMaterialnumbers[Decenas], UnidadesMaterial:=AllMaterialnumbers[Unidades]):
                CreativeProp.SetMaterial(UnidadesMaterial,?Index:=3)
                CreativeProp.SetMaterial(DecenasMaterial,?Index:=4)
                Button.SetInteractionText(BasicMessage("Teleport to the factory. Cost: x{IntAmount} Coal"))


    TryProduce_AllProducts() : void =
        CheckAsyncFame()
        TryProduce_B() 
        TryProduce_CB()
        TryProduce_PCS()
        CheckBakersMaterials()
        UpdateDroneDepositBillBoard()
    #Products
    TryProduce_B() : void =
        #Log("Progress_ForB: {Progress_ForB}",LOG_LEVEL_BAKERS)
        #if(MaxB_Storage > PlayerDataCache.B_On_Drone):
            if(Progress_ForB = 0.0):
                if(Req_WaterFor_B <= PlayerDataCache.WaterPreparedFor_B, Req_FlourFor_B <= PlayerDataCache.FlourPreparedFor_B):
                    spawn:
                        GameManager.VFXManager.AnimateBake(BB_Prop.GetTransform(),BB_Speed/SuperEnergy_Multiplier,false,B_TYPE)
                    set B_BachProduction = BB_Production
                    set Progress_ForB = Progress_ForB + (1.0/BB_Speed)
                    SetNewPlayerDataCache(PlayersDataManager.SetAfterB_SubDeposits(PlayerDataCache,PlayerDataCache.WaterPreparedFor_B - Req_WaterFor_B,PlayerDataCache.FlourPreparedFor_B - Req_FlourFor_B))
                    BB_Prop.SetMaterial(Mat_GreenBakerState, ?Index:=1)
                else:
                    BB_Prop.SetMaterial(Mat_RedBakerState, ?Index:=1)
            else:
                set Progress_ForB = Progress_ForB + (1.0/BB_Speed)*SuperEnergy_Multiplier
                #Log("Progress for Bread: {Progress_ForB} / 100", LOG_LEVEL_BAKERS)
                if(Progress_ForB >= 1.0):
                    AddProducts(B_BachProduction*SuperEnergy_Multiplier,B_TYPE)
                    #spawn:
                    #    VFXManager.AnimateShip(BB_Prop.GetTransform(),BB_Speed,false,B_TYPE)
                    set Progress_ForB = 0.0
                    set B_BachProduction = 0.0
                    ##Print("Produced Breads: {B_On_Drone}")

        if(BBInfoIsDisplayed?, not IsOnUKTutorial?, BBInfoUI := MaybeBBInfoUI?, AgentOwner:=MaybeAgentOwner?, PlayerOwner:=player[AgentOwner]<#, PlayerData := PlayersDataManager.GetPlayerData[PlayerOwner]#>):
            BBInfoUI.SetProductionInfo(BB_Production, CalculeBakerSpeed(BB_TYPE, PlayerDataCache.LV_BBSpeed, 1)(0))
            BBInfoUI.SetProgressForB(Progress_ForB, B_BachProduction*SuperEnergy_Multiplier)

    TryProduce_CB() : void =
        #Log("Progress_ForCB: {Progress_ForCB}",LOG_LEVEL_BAKERS)
        #if(MaxCB_Storage > PlayerDataCache.CB_On_Drone):
            if(Progress_ForCB = 0.0):
                if(Req_WaterFor_CB <= PlayerDataCache.WaterPreparedFor_CB, Req_FlourFor_CB <= PlayerDataCache.FlourPreparedFor_CB, Req_CheeseFor_CB <= PlayerDataCache.CheesePreparedFor_CB):
                    spawn:
                        GameManager.VFXManager.AnimateBake(CBB_Prop.GetTransform(),CBB_Speed/SuperEnergy_Multiplier,false,CB_TYPE)
                    set CB_BachProduction = CBB_Production
                    set Progress_ForCB = Progress_ForCB + (1.0/CBB_Speed)
                    SetNewPlayerDataCache(PlayersDataManager.SetAfterCB_SubDeposits(PlayerDataCache,PlayerDataCache.WaterPreparedFor_CB - Req_WaterFor_CB,PlayerDataCache.FlourPreparedFor_CB - Req_FlourFor_CB, PlayerDataCache.CheesePreparedFor_CB - Req_CheeseFor_CB))
                    #set WaterPreparedFor_CB = WaterPreparedFor_CB - Req_WaterFor_CB
                    #set FlourPreparedFor_CB = FlourPreparedFor_CB - Req_FlourFor_CB
                    #set CheesePreparedFor_CB = CheesePreparedFor_CB - Req_CheeseFor_CB
                    CBB_Prop.SetMaterial(Mat_GreenBakerState, ?Index:=1)
                else if(IsCBBAvailable?):
                    CBB_Prop.SetMaterial(Mat_RedBakerState, ?Index:=1)
            else:
                set Progress_ForCB = Progress_ForCB + (1.0/CBB_Speed)*SuperEnergy_Multiplier
                #Log("Progress for Bread: {Progress_ForB} / 100", LOG_LEVEL_BAKERS)
                if(Progress_ForCB >= 1.0):
                    AddProducts(CB_BachProduction*SuperEnergy_Multiplier,CB_TYPE)
                    #spawn:
                    #    VFXManager.AnimateShip(CBB_Prop.GetTransform(),CBB_Speed,false,CB_TYPE)
                    set Progress_ForCB = 0.0
                    set CB_BachProduction = 0.0
                    ##Print("Produced Breads: {CB_On_Drone}")

        if(CBBInfoIsDisplayed?, CBBInfoUI := MaybeCBBInfoUI?, AgentOwner:=MaybeAgentOwner?, PlayerOwner:=player[AgentOwner]<#, PlayerData := PlayersDataManager.GetPlayerData[PlayerOwner]#>):
            CBBInfoUI.SetProductionInfo(CBB_Production, CalculeBakerSpeed(CBB_TYPE, PlayerDataCache.LV_CBBSpeed, 1)(0))
            CBBInfoUI.SetProgressForCBB(Progress_ForCB, CB_BachProduction*SuperEnergy_Multiplier)

    TryProduce_PCS() : void =
        #Log("Progress_ForPCS: {Progress_ForPCS}",LOG_LEVEL_BAKERS)
        #if(MaxPCS_Storage > PlayerDataCache.PCS_On_Drone):
            if(Progress_ForPCS = 0.0):
                if(Req_WaterFor_PCS <= PlayerDataCache.WaterPreparedFor_PCS, Req_FlourFor_PCS <= PlayerDataCache.FlourPreparedFor_PCS, Req_CheeseFor_PCS <= PlayerDataCache.CheesePreparedFor_PCS, Req_PimentosFor_PCS <= PlayerDataCache.PimentosPreparedFor_PCS):
                    spawn:
                        GameManager.VFXManager.AnimateBake(PCSB_Prop.GetTransform(),PCSB_Speed/SuperEnergy_Multiplier,false,PCS_TYPE)
                    set PCS_BachProduction = PCSB_Production
                    set Progress_ForPCS = Progress_ForPCS + (1.0/PCSB_Speed)
                    SetNewPlayerDataCache(PlayersDataManager.SetAfterPCS_SubDeposits(PlayerDataCache,PlayerDataCache.WaterPreparedFor_PCS - Req_WaterFor_PCS,PlayerDataCache.FlourPreparedFor_PCS - Req_FlourFor_PCS, PlayerDataCache.CheesePreparedFor_PCS - Req_CheeseFor_PCS, PlayerDataCache.PimentosPreparedFor_PCS - Req_PimentosFor_PCS))
                    #set WaterPreparedFor_PCS = WaterPreparedFor_PCS - Req_WaterFor_PCS
                    #set FlourPreparedFor_PCS = FlourPreparedFor_PCS - Req_FlourFor_PCS
                    #set CheesePreparedFor_PCS = CheesePreparedFor_PCS - Req_CheeseFor_PCS
                    #set PimentosPreparedFor_PCS = PimentosPreparedFor_PCS - Req_PimentosFor_PCS
                    PCSB_Prop.SetMaterial(Mat_GreenBakerState, ?Index:=1)
                else if(IsPCSBAvailable?):
                    PCSB_Prop.SetMaterial(Mat_RedBakerState, ?Index:=1)
            else:
                set Progress_ForPCS = Progress_ForPCS + (1.0/PCSB_Speed)*SuperEnergy_Multiplier
                if(Progress_ForPCS >= 1.0):
                    AddProducts(PCS_BachProduction*SuperEnergy_Multiplier,PCS_TYPE)
                    #spawn:
                    #    VFXManager.AnimateShip(PCSB_Prop.GetTransform(),PCSB_Speed,false,PCS_TYPE)
                    set Progress_ForPCS = 0.0
                    set PCS_BachProduction = 0.0
                    ##Print("Produced Breads: {PCS_On_Drone}")

        if(PCSBInfoIsDisplayed?, PCSBInfoUI := MaybePCSBInfoUI?, AgentOwner:=MaybeAgentOwner?, PlayerOwner:=player[AgentOwner]<#, PlayerData := PlayersDataManager.GetPlayerData[PlayerOwner]#>):
            PCSBInfoUI.SetProductionInfo(PCSB_Production, CalculeBakerSpeed(PCSB_TYPE, PlayerDataCache.LV_PCSBSpeed, 1)(0))
            PCSBInfoUI.SetProgressForPCS(Progress_ForPCS, PCS_BachProduction*SuperEnergy_Multiplier)
                    

    AddProducts(Ammount : float, ProductType : int) : void =
        if(Agent := MaybeAgentOwner?, Player := player[Agent]<#, PlayerData := PlayersDataManager.GetPlayerData[Player]#>):
            if(ProductType=B_TYPE):
                var New_On_Drone : float = PlayerDataCache.B_On_Drone + Ammount
                #if(New_On_Drone > MaxB_Storage):
                #    set New_On_Drone = MaxB_Storage
                SetNewPlayerDataCache(PlayersDataManager.SetProductToDrone(PlayerDataCache, New_On_Drone, B_TYPE))
            else if(ProductType=CB_TYPE):
                var New_On_Drone : float = PlayerDataCache.CB_On_Drone + Ammount
                #if(New_On_Drone > MaxCB_Storage):
                #    set New_On_Drone = MaxCB_Storage
                SetNewPlayerDataCache(PlayersDataManager.SetProductToDrone(PlayerDataCache, New_On_Drone, CB_TYPE))
            else  if(ProductType=PCS_TYPE):
                var New_On_Drone : float = PlayerDataCache.PCS_On_Drone + Ammount
                #if(New_On_Drone > MaxPCS_Storage):
                #    set New_On_Drone = MaxPCS_Storage
                SetNewPlayerDataCache(PlayersDataManager.SetProductToDrone(PlayerDataCache, New_On_Drone, PCS_TYPE))
        UpdateDroneDepositBillBoard()
        

    
    InitBakery() : void =
        #ONLY INIT ONE TIME!
        #UIs
        #PCSBaker UI:
        if(Agent := MaybeAgentOwner?):
            OwnerReferenceDevice.Register(Agent)
            MasteryUI:=MakeMasteryUI(Self)
            MasteryUI.InitUI()
            set MaybeMasteryUI = option{MasteryUI}
            #PCSBInfoUI
            PCSBInfoUI:=MakePCSBInfoUI(Self)
            PCSBInfoUI.InitUI()
            set MaybePCSBInfoUI = option{PCSBInfoUI}
            #CBBInfoUI
            CBBInfoUI:=MakeCBBInfoUI(Self)
            #CBBInfoUI.SetAgent(Agent)
            CBBInfoUI.InitUI()
            set MaybeCBBInfoUI = option{CBBInfoUI}
            #BBInfoUI
            BBInfoUI:=MakeBBInfoUI(Self)
            BBInfoUI.InitUI()
            set MaybeBBInfoUI = option{BBInfoUI}
            #BB
            BBakersUpgradesUI:=MakeBBakerUpgradeUI(Self)
            BBakersUpgradesUI.SetAgent(Agent)
            BBakersUpgradesUI.InitUI()
            set MaybeBBakersUpgradesUI = option{BBakersUpgradesUI}
            #CBB
            CBBakersUpgradesUI:=MakeCBBakerUpgradeUI(Self)
            CBBakersUpgradesUI.SetAgent(Agent)
            CBBakersUpgradesUI.InitUI()
            set MaybeCBBakersUpgradesUI = option{CBBakersUpgradesUI}
            #PCSB
            PCSBakersUpgradesUI:=MakePCSBakerUpgradeUI(Self)
            PCSBakersUpgradesUI.SetAgent(Agent)
            PCSBakersUpgradesUI.InitUI()
            set MaybePCSBakersUpgradesUI = option{PCSBakersUpgradesUI}
            #DEPOSITS
            DepositsUpgradeUI:=MakeDepositsUpgradeUI(Self)
            DepositsUpgradeUI.SetAgent(Agent)
            DepositsUpgradeUI.InitUI()
            set MaybeDepositsUpgradeUI = option{DepositsUpgradeUI}
            #DRONES UPGRADES
            DronesUpgradesUI:=MakeDroneUpgradeUI(Self)
            DronesUpgradesUI.InitUI()
            set MaybeDroneUpgrades = option{DronesUpgradesUI}
            #BACKPACK UPGRADES
            BackpackUpgradesUI:=MakeBackpackUpgradeUI(Self)
            BackpackUpgradesUI.SetAgent(Agent)
            BackpackUpgradesUI.InitUI()
            set MaybeBackpackUpgradesUI = option{BackpackUpgradesUI}
            #ACTION INFO:
            ActionInfoUI:=MakeActionInfoUI(Self)
            ActionInfoUI.SetAgent(Agent)
            ActionInfoUI.InitUI()
            set MaybeActionInfoUI = option{ActionInfoUI}
            # Entity Health
            MainEntityHealth := MakeEntityHealthBarUI(Self)
            MainEntityHealth.SetAgent(Agent)
            MainEntityHealth.InitUI()
            MainEntityHealth.Display_Canvas()
            set MaybeEntityHealthUI = option{MainEntityHealth}
            # Inventory UI
            MainInventory := MakeMainInventoryUI(Self)
            MainInventory.SetAgent(Agent)
            MainInventory.InitUI()
            set MaybeMainInventoryUI = option{MainInventory}
            # Pickaxe UI
            PickaxeUpgradesUI := MakePickaxeUpgradeUI(Self)
            PickaxeUpgradesUI.SetAgent(Agent)
            PickaxeUpgradesUI.InitUI()
            set MaybePickaxeUpgradesUI = option{PickaxeUpgradesUI}
            # Resourcers UI
            #ResourcersUI := MakeResourersUI(Self)
            #ResourcersUI.SetAgent(Agent)
            #ResourcersUI.InitUI()
            #set MaybeResourcersUI = option{ResourcersUI}
            # Messages UI
            MessagesUI := MakeMessagesUI(Self)
            MessagesUI.SetAgent(Agent)
            MessagesUI.InitUI()
            set MaybeMessagesUI = option{MessagesUI}
            # Pickaxe Perks UI
            PickaxePerksUI := MakePickaxePerksUpgradeUI(Self)
            PickaxePerksUI.SetAgent(Agent)
            PickaxePerksUI.InitUI()
            set MaybePickaxePerksUI = option{PickaxePerksUI}
            #Weapon Upgrades
            WeaponsUpgradesUI:=MakeWeaponsUpgradeUI(Self)
            WeaponsUpgradesUI.SetAgent(Agent)
            WeaponsUpgradesUI.InitUI()
            WeaponsUpgradesUI.CreateWeaponStats()
            set MaybeWeaponsUpgradesUI = option{WeaponsUpgradesUI}
            #DungeonUI
            DungeonUI:=MakeDungeonUI(Self)
            DungeonUI.SetAgent(Agent)
            DungeonUI.InitUI()
            set MaybeDungeonUI = option{DungeonUI}
            #CoalUI
            DropCoalUI:=MakeDropCoalUI(Self)
            DropCoalUI.SetAgent(Agent)
            DropCoalUI.InitUI()
            set MaybeDropCoalUI = option{DropCoalUI}
            #TokenUI
            TokensUI:=MakeTokensUI(Self)
            TokensUI.SetAgent(Agent)
            TokensUI.InitUI()
            set MaybeTokensUI = option{TokensUI}
            #OptionsMenu
            OptionsMenuUI:=MakeOptionsMenuUI(Self)
            OptionsMenuUI.SetAgent(Agent)
            OptionsMenuUI.InitUI()
            set MaybeOptionsMenuUI = option{OptionsMenuUI}
            #Language
            #OptionsMenu
            LanguageUI:=MakeLanguageUI(Self)
            LanguageUI.InitUI()
            set MaybeLanguageUI = option{LanguageUI}
            #FastTravel
            FastTravelUI:=MakeFastTravelUI(Self)
            FastTravelUI.SetAgent(Agent)
            FastTravelUI.InitUI()
            set MaybeFastTravelUI = option{FastTravelUI}
            #Set HexaGameManager
            HexaGameManager.SetBakeryManager(Self)
            HexaGameManager.SetAgent(Agent)
            HexaGameManager.Init()
        InitSubscribes()
        DisplayBasicUIs()
        PetsManager.InitUIs()
        #Calcule inicial production
        spawn:
            LoadPlayerStats()
        if(AgentOwner:=MaybeAgentOwner?):
            spawn:
                GameManager.GeneralMobsManager.GranAgentWeapons(AgentOwner)
            if(DropCoalUI:=MaybeDropCoalUI?):
                DropCoalUI.TryIntroduceCoal_Amount(0.0)

    SetPropTransforms() : void = 
        set DroneInitialTransform = DroneProp.GetTransform()
        set FinalPosition = vector3{X:=DroneInitialTransform.Translation.X,Y:=DroneInitialTransform.Translation.Y+4096.0,Z:=DroneInitialTransform.Translation.Z}

    CheckBakersMaterials() : void =
        if(PlayerDataCache.NumQuestsCompleted>=17):
            CBB_Prop.SetMaterial(Materials.MI_BakerBaseMat,?Index:=0)
            CBB_Prop.SetMaterial(Materials.MI_Complement_red,?Index:=2)
            CBB_Prop.SetMaterial(Materials.Glass,?Index:=2)
            CBB_Prop.SetMaterial(Materials.MI_Complement_2,?Index:=3)
            set IsCBBAvailable = true
        else:
            set IsCBBAvailable = false
            CBB_Prop.SetMaterial(Materials.MI_BarrierBlue,?Index:=0)
            CBB_Prop.SetMaterial(Materials.MI_BarrierBlue,?Index:=1)
            CBB_Prop.SetMaterial(Materials.MI_BarrierBlue,?Index:=2)
            CBB_Prop.SetMaterial(Materials.MI_BarrierBlue,?Index:=3)
        if(PlayerDataCache.NumQuestsCompleted>=24):
            set IsPCSBAvailable = true
            PCSB_Prop.SetMaterial(Materials.MI_BakerBaseMat,?Index:=0)
            PCSB_Prop.SetMaterial(Materials.MI_Complement_red,?Index:=2)
            PCSB_Prop.SetMaterial(Materials.Glass,?Index:=2)
            PCSB_Prop.SetMaterial(Materials.MI_Complement_2,?Index:=3)
        else:
            set IsPCSBAvailable = false
            PCSB_Prop.SetMaterial(Materials.MI_BarrierBlue,?Index:=0)
            PCSB_Prop.SetMaterial(Materials.MI_BarrierBlue,?Index:=1)
            PCSB_Prop.SetMaterial(Materials.MI_BarrierBlue,?Index:=2)
            PCSB_Prop.SetMaterial(Materials.MI_BarrierBlue,?Index:=3)

    LoadPlayerStats(?IsMastery : logic =  false)<suspends> : void = 
        set IsBakerReady = false
        Sleep(0.08)
        if(Agent:=MaybeAgentOwner?, Player:=player[Agent], OldPlayerData := PlayersDataManager.LoadPlayerCache[Player]):
            if(not IsMastery?):
                SetNewPlayerDataCache(OldPlayerData)
            AllLV_PickaxePerks : []int = array{0,0,0,0,1} #CriticalProb,CriticalDamage,PickaxeLuck,PickaxeWeakPointMult,PickaxeMultiResource       
            #Production settings
            Sleep(0.08)
            #set SuperEnergy_Owned = PlayerData.SuperEnergy_Owned
            set BB_Production = CalculeTotalProduction(PlayerDataCache.TLV_BB,BB_TYPE)
            set CBB_Production = CalculeTotalProduction(PlayerDataCache.TLV_CBB,CBB_TYPE)
            set PCSB_Production = CalculeTotalProduction(PlayerDataCache.TLV_PCSB,PCSB_TYPE)
            #Baker Materials
            CheckBakersMaterials()
            #Pickaxe settings
            SetPickaxePerks(PlayerDataCache)
            set PickaxeTotalStrength=CalculeTotalStrength(PlayerDataCache.TLV_PickaxeStrength)
            <#
            #Products on drone
            set B_On_Drone = PlayerData.B_On_Drone
            set CB_On_Drone = PlayerData.CB_On_Drone
            set PCS_On_Drone = PlayerData.PCS_On_Drone
            #SetDeposits
            set MainWaterDeposit_Value = PlayerData.MainWaterDeposit_Value
            set MainFlourDeposit_Value = PlayerData.MainFlourDeposit_Value
            set MainCheeseDeposit_Value = PlayerData.MainCheeseDeposit_Value
            set MainPimentosDeposit_Value = PlayerData.MainPimentosDeposit_Value
            #Subdeposits
            set WaterPreparedFor_B = PlayerData.WaterPreparedFor_B
            set FlourPreparedFor_B = PlayerData.FlourPreparedFor_B
            set WaterPreparedFor_CB = PlayerData.WaterPreparedFor_CB
            set FlourPreparedFor_CB = PlayerData.FlourPreparedFor_CB
            set CheesePreparedFor_CB = PlayerData.CheesePreparedFor_CB
            set WaterPreparedFor_PCS = PlayerData.WaterPreparedFor_PCS
            set FlourPreparedFor_PCS = PlayerData.FlourPreparedFor_PCS
            set CheesePreparedFor_PCS = PlayerData.CheesePreparedFor_PCS
            set PimentosPreparedFor_PCS = PlayerData.PimentosPreparedFor_PCS#>
            #Pickaxe efficiency
            set PickaxeTotalStrength = CalculeTotalStrength(PlayerDataCache.TLV_PickaxeStrength)
            GenerateAllBackPackLimits(PlayerDataCache)
            Sleep(0.08)
            SetPickaxeEfficiency(Agent,PlayerDataCache,WATER)
            SetPickaxeEfficiency(Agent,PlayerDataCache,FLOUR)
            SetPickaxeEfficiency(Agent,PlayerDataCache,CHEESE)
            SetPickaxeEfficiency(Agent,PlayerDataCache,PIMENTOS)
            SetPickaxeEfficiency(Agent,PlayerDataCache,MUTANTMATERIAL)
            SetScrollPickaxeEfficiency(PlayerDataCache)

            #Baker efficiency
            Log("Setting efficiency", LOG_LEVEL_UIs)
            SetBakerEfficiency(PlayerDataCache,BB_TYPE)
            SetBakerEfficiency(PlayerDataCache,CBB_TYPE)
            SetBakerEfficiency(PlayerDataCache,PCSB_TYPE)

            SetAllMainDepositLimits(Agent,PlayerDataCache.AllLV_MainDeposits)
            #Drone settings
            SetDroneSettings(PlayerDataCache,true)
            SetSalesEfficiency(PlayerDataCache,B_TYPE)
            SetSalesEfficiency(PlayerDataCache,CB_TYPE)
            SetSalesEfficiency(PlayerDataCache,PCS_TYPE)
            Sleep(0.08)
            set BB_Speed = CalculeBakerSpeed(BB_TYPE,PlayerDataCache.LV_BBSpeed,1)(0)
            set CBB_Speed = CalculeBakerSpeed(CBB_TYPE,PlayerDataCache.LV_CBBSpeed,1)(0)
            set PCSB_Speed = CalculeBakerSpeed(PCSB_TYPE,PlayerDataCache.LV_PCSBSpeed,1)(0)
            set BB_Production = CalculeTotalProduction(PlayerDataCache.TLV_BB,BB_TYPE)
            set CBB_Production = CalculeTotalProduction(PlayerDataCache.TLV_CBB,CBB_TYPE)
            set PCSB_Production = CalculeTotalProduction(PlayerDataCache.TLV_PCSB,PCSB_TYPE)
            RefreshInventoryUI(PlayerDataCache)
            InitPetManager(Agent,PlayerDataCache)
            PetsManager.Generate_PlayerEnhacements()
            Sleep(0.08)
            GameManager.GeneralMobsManager.GranAgentWeapons(Agent)
            if(PlayerDataCache.IsTutorialCompleted? or PlayerDataCache.Inventory_Fame > 100.0):
                spawn:
                    GameManager.QuestsAndOrdersDatabase.DelayCheckWelcomeLore(Player,Self,2.0)
                if(not PlayerDataCache.IsTutorialCompleted?):
                    SetNewPlayerDataCache(PlayersDataManager.SetTutorialState(PlayerDataCache,true))
            else:
                if(not GameManager.DebugOn?):
                    GameManager.EnterOnUKTutorial(Agent)
            if(IsMastery?):
                #Valves
                #States
                set ValveState_WaterFor_B = 1.0
                set ValveState_FlourFor_B = 1.0
                UpdateValveProps(0, 1.0)
                UpdateValveProps(1, 1.0)

                set ValveState_WaterFor_CB = 1.0
                set ValveState_FlourFor_CB = 1.0
                set ValveState_CheeseFor_CB = 1.0

                UpdateValveProps(2, 1.0)
                UpdateValveProps(3, 1.0)
                UpdateValveProps(4, 1.0)

                set ValveState_WaterFor_PCS = 1.0
                set ValveState_FlourFor_PCS = 1.0
                set ValveState_CheeseFor_PCS = 1.0
                set ValveState_PimentosFor_PCS = 1.0

                UpdateValveProps(6, 1.0)
                UpdateValveProps(5, 1.0)
                UpdateValveProps(7, 1.0)

                UpgradeValveButtonsDescriptions()
        else:
            PrintFatalError("On LoadPlayerStats()")
        #SetPasiveIncome
        SetAllPasiveResourcers()
        CheckQuest()
        if(WeaponsUpgradesUI:=MaybeWeaponsUpgradesUI?):
            WeaponsUpgradesUI.CreateWeaponStats()
        set IsBakerReady = true
       
    CheckQuestProgress(AmountProgress : float, TypeQuest : int) : void =
        if(AgentOwner := MaybeAgentOwner?, Player := player[AgentOwner], not IsOnUKTutorial?<#, PlayerData:=PlayersDataManager.GetPlayerData[Player]#>):
            SetNewPlayerDataCache(GameManager.QuestsAndOrdersDatabase.AddAgentQuest_ProgressV2(Player,PlayerDataCache,AmountProgress,TypeQuest)(4))

    UnlockCBB() : void =
        CBB_Prop.SetMaterial(Materials.MI_BakerBaseMat,?Index:=0)
        if(Progress_ForCB = 0.0):
            CBB_Prop.SetMaterial(Materials.MI_Complement_red,?Index:=2)
        CBB_Prop.SetMaterial(Materials.Glass,?Index:=2)
        CBB_Prop.SetMaterial(Materials.MI_Complement_2,?Index:=3)
        set IsCBBAvailable = true

    UnlockPCSB() : void =
        PCSB_Prop.SetMaterial(Materials.MI_BakerBaseMat,?Index:=0)
        if(Progress_ForPCS = 0.0):
            PCSB_Prop.SetMaterial(Materials.MI_Complement_red,?Index:=2)
        PCSB_Prop.SetMaterial(Materials.Glass,?Index:=2)
        PCSB_Prop.SetMaterial(Materials.MI_Complement_2,?Index:=3)
        set IsPCSBAvailable = true
    

    InitPetManager(Agent : agent, PlayerData : player_data_table) : void =
        PetsManager.CreateUpgradesUI()
        PetsManager.SetAgent(Agent)
        #PetsManager.InitUIs()
        PetsManager.Generate_PlayerEnhacements()
        PetsManager.LoadEquippedPets(PlayerData)

    CompleteQuest() : void =
        if(AgentOwner:=MaybeAgentOwner?,Player := player[AgentOwner], Player.IsActive[]<#, PlayerData := PlayersDataManager.GetPlayerData[Player]#>, MainInventoryUI:=MaybeMainInventoryUI?, ActionUI:=MaybeActionInfoUI?):
            #GameManager.DisplayCharacterMessage(AgentOwner, W_CHARACTER, "Tutorial completed! Now complete the next")
            QuestInfoResult:=GameManager.QuestsAndOrdersDatabase.AddAgentQuest_ProgressV2(Player,PlayerDataCache,1.0,PlayerDataCache.QuestAssigned_ObjetiveType)(4)
            SetNewPlayerDataCache(QuestInfoResult)

    SetPickaxePerks(PlayerData : player_data_table) : void =
        if:
            LV_PickaxeCriticalProb := PlayerData.AllLV_PickaxePerks[CRITICAL_PROB]
            LV_PickaxeCriticalDamage := PlayerData.AllLV_PickaxePerks[CRITICAL_DAMAGE]
            LV_PickaxeLuck := PlayerData.AllLV_PickaxePerks[PICKAXE_LUCK]
            LV_PickaxeWeakPointMult := PlayerData.AllLV_PickaxePerks[PICKAXE_WEAK_MULT]
            LV_PickaxeMultiResource := PlayerData.AllLV_PickaxePerks[PICKAXE_MULT]
        then:
            set PickaxeCriticalProb = CalculePickaxeCriticalProbUpgrade(LV_PickaxeCriticalProb)(0)
            set PickaxeCriticalDamage = CalculePickaxeCriticalDamageUpgrade(LV_PickaxeCriticalDamage)(0)
            set PickaxeCoalLuck = CalculePickaxeLuckUpgrade(LV_PickaxeLuck)(0)
            set PickaxeWeakPointMult = CalculePickaxeWeakUpgrade(LV_PickaxeWeakPointMult)(0)
            FloatMultipleResUpgrade := CalculeMultipleResUpgrade(LV_PickaxeMultiResource)(0)
            if(Value := Floor[FloatMultipleResUpgrade]):
                set PickaxeMultiResource = Value
        
    SetDroneSettings(PlayerData : player_data_table, LoadMode : logic) : void =
     
        set LV_DroneSpeed = PlayerData.LV_DroneSpeed
        set LV_DroneStorage = PlayerData.LV_DroneStorage

        set AllLV_SalesEfficiency = PlayerData.AllLV_SalesEfficiency


        #Set values
        set DroneSpeed = CalculeDroneSpeed(LV_DroneSpeed)(0)

        SetDroneMaxStorages()

        if(LoadMode?):
            set TimeToDroneReturn = PlayerData.TimeToDroneReturn
            if(TimeToDroneReturn>0):
                set IsDronePropOnBase = false
                FinalTransform := transform{Translation:=FinalPosition,Rotation:=rotation{},Scale:=vector3{X:=1.0,Y:=1.0,Z:=1.0}}
                if(DroneProp.TeleportTo[FinalTransform]):
            else if(TimeToDroneReturn = 0):
                DroneShip_Billbord.SetText(BasicMessage("READY FOR SHIPMENT"))
                set IsDronePropOnBase = true
                if(DroneProp.TeleportTo[DroneInitialTransform]):

    SetDroneMaxStorages() : void =
        UpdateDroneDepositBillBoard()
        <#
        FloatNoConverted_MaxB_Storage := CalculeDroneStorage(BB_Production, BB_Speed)(0)
        FloatNoConverted_MaxCB_Storage := CalculeDroneStorage(CBB_Production, CBB_Speed)(0)
        FloatNoConverted_MaxPCS_Storage := CalculeDroneStorage(PCSB_Production, PCSB_Speed)(0)

        if(FloatNoConverted_MaxB_Storage<1000.0, IntConverted_MaxB_Storage := Floor[FloatNoConverted_MaxB_Storage]):
            set MaxB_Storage = IntConverted_MaxB_Storage *1.0
        else:
            set MaxB_Storage = FloatNoConverted_MaxB_Storage

        if(FloatNoConverted_MaxCB_Storage<1000.0, IntConverted_MaxCB_Storage := Floor[FloatNoConverted_MaxCB_Storage]):
            set MaxCB_Storage = IntConverted_MaxCB_Storage *1.0
        else:
            set MaxCB_Storage = FloatNoConverted_MaxCB_Storage

        if(FloatNoConverted_MaxPCS_Storage<1000.0, IntConverted_MaxPCS_Storage := Floor[FloatNoConverted_MaxPCS_Storage]):
            set MaxPCS_Storage = IntConverted_MaxPCS_Storage *1.0
        else:
            set MaxPCS_Storage = FloatNoConverted_MaxPCS_Storage

        #set MaxB_Storage = FloatNoConverted_MaxB_Storage
        #set MaxCB_Storage = FloatNoConverted_MaxCB_Storage
        #set MaxPCS_Storage = FloatNoConverted_MaxPCS_Storage
        UpdateDroneDepositBillBoard()
        #>

    UpdateDroneDepositBillBoard() : void =
        if(IsDronePropOnBase?):
            DroneDeposit_Billbord.SetText(BasicMessage("{GetOptimizedMagnitudeLetter(PlayerDataCache.B_On_Drone)}\n{GetOptimizedMagnitudeLetter(PlayerDataCache.CB_On_Drone)}\n{GetOptimizedMagnitudeLetter(PlayerDataCache.PCS_On_Drone)}"))             
        else:
            DroneDeposit_Billbord.SetText(BasicMessage("{GetOptimizedMagnitudeLetter(PlayerDataCache.B_On_Drone)} [ON HOLD]\n{GetOptimizedMagnitudeLetter(PlayerDataCache.CB_On_Drone)} [ON HOLD]\n{GetOptimizedMagnitudeLetter(PlayerDataCache.PCS_On_Drone)} [ON HOLD]"))             
        <#
        if(IsDronePropOnBase?):
            DroneDeposit_Billbord.SetText(BasicMessage("{GetOptimizedMagnitudeLetter(PlayerDataCache.B_On_Drone)} / {GetOptimizedMagnitudeLetter(MaxB_Storage)}\n{GetOptimizedMagnitudeLetter(PlayerDataCache.CB_On_Drone)} / {GetOptimizedMagnitudeLetter(MaxCB_Storage)}\n{GetOptimizedMagnitudeLetter(PlayerDataCache.PCS_On_Drone)} / {GetOptimizedMagnitudeLetter(MaxPCS_Storage)}"))             
        else:
            DroneDeposit_Billbord.SetText(BasicMessage("{GetOptimizedMagnitudeLetter(PlayerDataCache.B_On_Drone)} / {GetOptimizedMagnitudeLetter(MaxB_Storage)} [ON HOLD]\n{GetOptimizedMagnitudeLetter(PlayerDataCache.CB_On_Drone)} / {GetOptimizedMagnitudeLetter(MaxCB_Storage)} [ON HOLD]\n{GetOptimizedMagnitudeLetter(PlayerDataCache.PCS_On_Drone)} / {GetOptimizedMagnitudeLetter(MaxPCS_Storage)} [ON HOLD]"))             
        #>

    SetBakerEfficiency(PlayerData : player_data_table, Type : int) : void =
        if(Type=BB_TYPE):
            #BB
            if:
                BB_WaterEfficiencyLV := PlayerData.EfficiencyLV_BB[0]
                BB_FlourEfficiencyLV := PlayerData.EfficiencyLV_BB[1]
            then:
                set BB_WaterEfficiency = 1.0+(CalculeBakerResourceEfficiency(BB_WaterEfficiencyLV, BB_TYPE, WATER, 1)(0)/100.0)
                set BB_FlourEfficiency = 1.0+(CalculeBakerResourceEfficiency(BB_FlourEfficiencyLV, BB_TYPE, FLOUR, 1)(0)/100.0)
                
            set Req_WaterFor_B = (Base_Req_WaterFor_B * BB_Production)/BB_WaterEfficiency
            set Req_FlourFor_B = (Base_Req_FlourFor_B * BB_Production)/BB_FlourEfficiency

        else if(Type=CBB_TYPE):
            #BBC
            if:
                CBB_WaterEfficiencyLV := PlayerData.EfficiencyLV_CBB[0]
                CBB_FlourEfficiencyLV := PlayerData.EfficiencyLV_CBB[1]
                CBB_CheeseEfficiencyLV := PlayerData.EfficiencyLV_CBB[2]
            then:
                set CBB_WaterEfficiency = 1.0+(CalculeBakerResourceEfficiency(CBB_WaterEfficiencyLV, CBB_TYPE, WATER, 1)(0)/100.0)
                set CBB_FlourEfficiency = 1.0+(CalculeBakerResourceEfficiency(CBB_FlourEfficiencyLV, CBB_TYPE, FLOUR, 1)(0)/100.0)
                set CBB_CheeseEfficiency = 1.0+(CalculeBakerResourceEfficiency(CBB_CheeseEfficiencyLV, CBB_TYPE, CHEESE, 1)(0)/100.0)

            set Req_WaterFor_CB = (Base_Req_WaterFor_CB * CBB_Production)/CBB_WaterEfficiency
            set Req_FlourFor_CB = (Base_Req_FlourFor_CB * CBB_Production)/CBB_FlourEfficiency
            set Req_CheeseFor_CB = (Base_Req_CheeseFor_CB * CBB_Production)/CBB_CheeseEfficiency
        else if(Type=PCSB_TYPE):
            #PCSB
            if:
                PCSB_WaterEfficiencyLV := PlayerData.EfficiencyLV_PCSB[0]
                PCSB_FlourEfficiencyLV := PlayerData.EfficiencyLV_PCSB[1]
                PCSB_CheeseEfficiencyLV := PlayerData.EfficiencyLV_PCSB[2]
                PCSB_PimentosEfficiencyLV := PlayerData.EfficiencyLV_PCSB[3]
            then:
                set PCSB_WaterEfficiency = 1.0+(CalculeBakerResourceEfficiency(PCSB_WaterEfficiencyLV, PCSB_TYPE, WATER, 1)(0)/100.0)
                set PCSB_FlourEfficiency = 1.0+(CalculeBakerResourceEfficiency(PCSB_FlourEfficiencyLV, PCSB_TYPE, FLOUR, 1)(0)/100.0)
                set PCSB_CheeseEfficiency = 1.0+(CalculeBakerResourceEfficiency(PCSB_CheeseEfficiencyLV, PCSB_TYPE, CHEESE, 1)(0)/100.0)
                set PCSB_PimentosEfficiency = 1.0+(CalculeBakerResourceEfficiency(PCSB_PimentosEfficiencyLV, PCSB_TYPE, PIMENTOS, 1)(0)/100.0)

            set Req_WaterFor_PCS = (Base_Req_WaterFor_PCS * PCSB_Production)/PCSB_WaterEfficiency
            set Req_FlourFor_PCS = (Base_Req_FlourFor_PCS * PCSB_Production)/PCSB_FlourEfficiency
            set Req_CheeseFor_PCS = (Base_Req_CheeseFor_PCS * PCSB_Production)/PCSB_CheeseEfficiency
            set Req_PimentosFor_PCS = (Base_Req_PimentosFor_PCS * PCSB_Production)/PCSB_PimentosEfficiency

            Log("Req_WaterFor_PCS: {Req_WaterFor_PCS}",LOG_LEVEL_UIs)
            Log("Req_FlourFor_PCS: {Req_FlourFor_PCS}",LOG_LEVEL_UIs)
            Log("Req_CheeseFor_PCS: {Req_CheeseFor_PCS}",LOG_LEVEL_UIs)
            Log("Req_PimentosFor_PCS: {Req_PimentosFor_PCS}",LOG_LEVEL_UIs)


    SetSalesEfficiency(PlayerData : player_data_table, Product : int) : void =
        if:
            ResourceEfficiencyLV := PlayerData.AllLV_SalesEfficiency[Product]
        then:
            ResourceEfficiency := 1.0+(CalculeSalesEfficiency(ResourceEfficiencyLV)(0)/100.0)
            if(set All_SalesEfficiency[Product]=ResourceEfficiency):
  

    #Valve functions
    UpgradeValveButtonsDescriptions() : void = #High load function!
        if(not GameManager.IsValveSimplifiedVersion?):
            if:
                WButtonValve_0 := AllButtons_Valves[0]
                WButtonValve_1 := AllButtons_Valves[2]
                WButtonValve_2 := AllButtons_Valves[5]
            then:
                WButtonValve_0.SetInteractionText(MessageBasic("💧 Change Valve"))
                WButtonValve_1.SetInteractionText(MessageBasic("💧 Change Valve"))
                WButtonValve_2.SetInteractionText(MessageBasic("💧 Change Valve"))

            if:
                FButtonValve_0 := AllButtons_Valves[1]
                FButtonValve_1 := AllButtons_Valves[3]
                FButtonValve_2 := AllButtons_Valves[6]
            then:
                FButtonValve_0.SetInteractionText(MessageBasic("🌾 Change Valve"))
                FButtonValve_1.SetInteractionText(MessageBasic("🌾 Change Valve"))
                FButtonValve_2.SetInteractionText(MessageBasic("🌾 Change Valve"))
            if:
                CButtonValve_0 := AllButtons_Valves[4]
                CButtonValve_1 := AllButtons_Valves[7]
            then:
                CButtonValve_0.SetInteractionText(MessageBasic("🧀 Change Valve"))
                CButtonValve_1.SetInteractionText(MessageBasic("🧀 Change Valve"))
        else:
            Simplified_ValveButtons_B.SetInteractionText(MessageBasic("🥖 Change Bread Baker valves"))
            Simplified_ValveButtons_CB.SetInteractionText(MessageBasic("🧀 Change Cheese Bread Baker valves"))
            Simplified_ValveButtons_PCS.SetInteractionText(MessageBasic("🥪 Change Pimento Cheese Sandwich Baker valves"))

        <#
        if(Button := AllButtons_Valves[0]):
            Button.SetInteractionText(MessageBasic(VALVEFLOW_BUTTON_WATER_MESSAGE+"{GetOptimizedMagnitudeLetter(WaterFlowPerSecond_ForB_ToTake)}/s"))
        if(Button := AllButtons_Valves[1]):
            Button.SetInteractionText(MessageBasic(VALVEFLOW_BUTTON_FLOUR_MESSAGE+"{GetOptimizedMagnitudeLetter(FlourFlowPerSecond_ForB_ToTake)}/s"))

        if(IsCBBAvailable?):
            if(Button := AllButtons_Valves[2]):
                Button.SetInteractionText(MessageBasic("Change value"))
            if(Button := AllButtons_Valves[3]):
                Button.SetInteractionText(MessageBasic(VALVEFLOW_BUTTON_FLOUR_MESSAGE+"{GetOptimizedMagnitudeLetter(FlourFlowPerSecond_ForCB_ToTake)}/s"))
            if(Button := AllButtons_Valves[4]):
                Button.SetInteractionText(MessageBasic(VALVEFLOW_BUTTON_CHEESE_MESSAGE+"{GetOptimizedMagnitudeLetter(CheeseFlowPerSecond_ForCB_ToTake)}/s"))

        if(IsPCSBAvailable?):
            if(Button := AllButtons_Valves[5]):
                Button.SetInteractionText(MessageBasic(VALVEFLOW_BUTTON_WATER_MESSAGE+"{GetOptimizedMagnitudeLetter(WaterFlowPerSecond_ForPCS_ToTake)}/s"))
            if(Button := AllButtons_Valves[6]):
                Button.SetInteractionText(MessageBasic(VALVEFLOW_BUTTON_FLOUR_MESSAGE+"{GetOptimizedMagnitudeLetter(FlourFlowPerSecond_ForPCS_ToTake)}/s"))
            if(Button := AllButtons_Valves[7]):
                Button.SetInteractionText(MessageBasic(VALVEFLOW_BUTTON_CHEESE_MESSAGE+"{GetOptimizedMagnitudeLetter(CheeseFlowPerSecond_ForPCS_ToTake)}/s"))
            #if(Button := AllButtons_Valves[8]):
            #    Button.SetInteractionText(MessageBasic(VALVEFLOW_BUTTON_PIMENTOS_MESSAGE+"{GetOptimizedMagnitudeLetter(PimentosFlowPerSecond_ForPCS_ToTake)}/s"))
        #>


    Change_ValveState(Agent : agent, Index : int) : void =
        ##Print("Trying to change valve state {Index}")
        if(AgentOwner := MaybeAgentOwner?, AgentOwner = Agent):
            var Value : float = -1.0
            var ValveCanBeChanged : logic = true
            if(Index>=2, not IsCBBAvailable?):
                set ValveCanBeChanged = false
            if(Index>=5, not IsPCSBAvailable?):
                set ValveCanBeChanged = false
            if(ValveCanBeChanged?):
                case(Index):
                    0=> set Value = ChangeValve_WaterFor_B(2.0)
                    1=> set Value = ChangeValve_FlourFor_B(2.0)
                    2=> set Value = ChangeValve_WaterFor_CB(2.0)
                    3=> set Value = ChangeValve_FlourFor_CB(2.0)
                    4=> set Value = ChangeValve_CheeseFor_CB(2.0)
                    5=> set Value = ChangeValve_WaterFor_PCS(2.0)
                    6=> set Value = ChangeValve_FlourFor_PCS(2.0)  
                    7=> set Value = ChangeValve_CheeseFor_PCS(2.0)  
                    #8=> set Value = ChangeValve_PimentosFor_PCS(2.0)
                    _=> #Print("FATAL ERROR On: Change_ValveState. Index Value: {Index} not valid")
                UpdateValveProps(Index,Value)
                CheckValveTutorial()
                GameManager.VFXManager.Audio_ValveChanged.Play(Agent)
            else:
                GameManager.DisplayCharacterMessage(Agent,I_CHARACTER,"You have not unlocked this Baker",?Sound:=ERROR_SOUND)
        else:
            GameManager.Display_NotOwner(Agent)

    Change_SimplifiedValveState(Agent : agent, BakeryType : int) : void = 
        if(AgentOwner := MaybeAgentOwner?, AgentOwner = Agent):
            var ValveCanBeChanged : logic = true
            if(BakeryType = CBB_TYPE, not IsCBBAvailable?):
                set ValveCanBeChanged = false
            if(BakeryType = PCSB_TYPE, not IsPCSBAvailable?):
                set ValveCanBeChanged = false
            if(ValveCanBeChanged?):
                if(BakeryType = BB_TYPE):
                    var Water : float = ChangeValve_WaterFor_B(2.0)
                    var Flour : float = ChangeValve_FlourFor_B(2.0)
                    if(Water<>Flour):
                        set Water = Flour
                    UpdateValveProps(0,Water)
                    UpdateValveProps(1,Flour)
                if(BakeryType = CBB_TYPE):
                    var Water : float = ChangeValve_WaterFor_CB(2.0)
                    var Flour : float = ChangeValve_FlourFor_CB(2.0)
                    var Cheese : float = ChangeValve_CheeseFor_CB(2.0)
                    if(Water<>Flour or Water<>Cheese):
                        set Water = Flour
                        set Cheese = Flour
                    UpdateValveProps(2,Water)
                    UpdateValveProps(3,Flour)
                    UpdateValveProps(4,Cheese)
                if(BakeryType = PCSB_TYPE):
                    var Water : float = ChangeValve_WaterFor_PCS(2.0)
                    var Flour : float = ChangeValve_FlourFor_PCS(2.0)
                    var Cheese : float = ChangeValve_CheeseFor_PCS(2.0)
                    var Pimentos : float = ChangeValve_PimentosFor_PCS(2.0)
                    if(Water<>Flour or Water<>Cheese or Water<>Pimentos):
                        set Water = Flour
                        set Cheese = Flour
                        set Pimentos = Flour
                    UpdateValveProps(5,Water)
                    UpdateValveProps(6,Flour)
                    UpdateValveProps(7,Cheese)
                CheckValveTutorial()
                GameManager.VFXManager.Audio_ValveChanged.Play(Agent)
                CheckQuestProgress(1.0,CHANGEVALVES_QUEST_TYPE)
            else:
                GameManager.DisplayCharacterMessage(Agent,I_CHARACTER,"You have not unlocked this Baker yet",?Sound:=ERROR_SOUND)
        else:
            GameManager.Display_NotOwner(Agent) 

    CheckValveTutorial() : void =
        if(ValveState_WaterFor_B>0.0,ValveState_FlourFor_B>0.0):
            CheckQuestProgress(1.0,USEVALVES_TUTORIAL_TYPE)
        if(ValveState_WaterFor_CB=2.0,ValveState_FlourFor_CB=2.0,ValveState_CheeseFor_CB=2.0):
            CheckQuestProgress(1.0,OPENINGVALVES_TUTORIAL_TYPE)
        if(ValveState_WaterFor_B=0.0,ValveState_FlourFor_B=0.0):
            CheckQuestProgress(1.0,CLOSINGVALVES_TUTORIAL_TYPE)
        if((ValveState_WaterFor_B=1.0,ValveState_FlourFor_B=1.0) or (ValveState_WaterFor_CB=1.0,ValveState_FlourFor_CB=1.0,ValveState_CheeseFor_CB=1.0)):
            CheckQuestProgress(1.0,HALFVALVES_TUTORIAL_TYPE)
        if(ValveState_WaterFor_PCS>0.0,ValveState_FlourFor_PCS>0.0,ValveState_CheeseFor_PCS>0.0):
            CheckQuestProgress(1.0,PCSBOPENINGVALVES_TUTORIAL_TYPE)
    

            #Bread
    ChangeValve_WaterFor_B(MaxValveValue : float) : float =
        if(ValveState_WaterFor_B >= MaxValveValue):
            set ValveState_WaterFor_B = 0.0
        else:
            set ValveState_WaterFor_B = ValveState_WaterFor_B + 1.0
        Log("ValveState_WaterFor_B: {ValveState_WaterFor_B}", LOG_LEVEL_BAKERS)
        return ValveState_WaterFor_B

    ChangeValve_FlourFor_B(MaxValveValue : float) : float =
        if(ValveState_FlourFor_B >= MaxValveValue):
            set ValveState_FlourFor_B = 0.0
        else:
            set ValveState_FlourFor_B = ValveState_FlourFor_B + 1.0
        return ValveState_FlourFor_B

    ChangeValve_WaterFor_CB(MaxValveValue : float) : float =
        if(ValveState_WaterFor_CB >= MaxValveValue):
            set ValveState_WaterFor_CB = 0.0
        else:
            set ValveState_WaterFor_CB = ValveState_WaterFor_CB + 1.0
        return ValveState_WaterFor_CB

    ChangeValve_FlourFor_CB(MaxValveValue : float) : float =
        if(ValveState_FlourFor_CB >= MaxValveValue):
            set ValveState_FlourFor_CB = 0.0
        else:
            set ValveState_FlourFor_CB = ValveState_FlourFor_CB + 1.0
        return ValveState_FlourFor_CB

    ChangeValve_CheeseFor_CB(MaxValveValue : float) : float =
        if(ValveState_CheeseFor_CB >= MaxValveValue):
            set ValveState_CheeseFor_CB = 0.0
        else:
            set ValveState_CheeseFor_CB = ValveState_CheeseFor_CB + 1.0
        return ValveState_CheeseFor_CB

    ChangeValve_WaterFor_PCS(MaxValveValue : float) : float =
        if(ValveState_WaterFor_PCS >= MaxValveValue):
            set ValveState_WaterFor_PCS = 0.0
        else:
            set ValveState_WaterFor_PCS = ValveState_WaterFor_PCS + 1.0
        return ValveState_WaterFor_PCS

    ChangeValve_FlourFor_PCS(MaxValveValue : float) : float =
        if(ValveState_FlourFor_PCS >= MaxValveValue):
            set ValveState_FlourFor_PCS = 0.0
        else:
            set ValveState_FlourFor_PCS = ValveState_FlourFor_PCS + 1.0
        return ValveState_FlourFor_PCS

    ChangeValve_CheeseFor_PCS(MaxValveValue : float) : float =
        if(ValveState_CheeseFor_PCS >= MaxValveValue):
            set ValveState_CheeseFor_PCS = 0.0
        else:
            set ValveState_CheeseFor_PCS = ValveState_CheeseFor_PCS + 1.0
        return ValveState_CheeseFor_PCS

    ChangeValve_PimentosFor_PCS(MaxValveValue : float) : float =
        if(ValveState_PimentosFor_PCS >= MaxValveValue):
            set ValveState_PimentosFor_PCS = 0.0
        else:
            set ValveState_PimentosFor_PCS = ValveState_PimentosFor_PCS + 1.0
        return ValveState_PimentosFor_PCS

    InitValves() : void =
        UpgradeValveButtonsDescriptions()
        set ValveState_WaterFor_B = 0.0
        set ValveState_FlourFor_B = 0.0
        UpdateValveProps(0, 0.0)
        UpdateValveProps(1, 0.0)

        #if(LV_CheeseBreadBaker>=0):
        set ValveState_WaterFor_CB = 0.0
        set ValveState_FlourFor_CB = 0.0
        set ValveState_CheeseFor_CB = 0.0
        UpdateValveProps(2, 0.0)
        UpdateValveProps(3, 0.0)
        UpdateValveProps(4, 0.0)

        #if(LV_PimentoCheeseBlender>=0):
        set ValveState_WaterFor_PCS = 0.0
        set ValveState_FlourFor_PCS = 0.0
        set ValveState_CheeseFor_PCS = 0.0
        set ValveState_PimentosFor_PCS = 0.0
        UpdateValveProps(6, 0.0)
        UpdateValveProps(5, 0.0)
        UpdateValveProps(7, 0.0)
        #UpdateValveProps(8, 0.0)
    
    #Displays
    #ComputerDisplay
    Display_PickaxeUpgrades(Agent : agent) : void =
        if(AgentOwner:=MaybeAgentOwner?, Agent=AgentOwner, PlayerOwner:=player[AgentOwner]):
            if:
                PickaxeUpgradesUI := MaybePickaxeUpgradesUI?
            then:
                PickaxeUpgradesUI.Display_Canvas()
                if(not IsOnUKTutorial?):
                    PickaxeUpgradesUI.UpdateUI(PlayerDataCache,PickaxeTotalStrength)
                else:
                    PickaxeUpgradesUI.UpdateUI(player_data_table{},CalculeTotalStrength(player_data_table{}.TLV_PickaxeStrength))
        else:
            GameManager.Display_NotOwner(Agent)

    #Pickaxe perks upgrades
    Display_PickaxePerksUpgrades(Agent : agent) : void =
        if(AgentOwner:=MaybeAgentOwner?, Agent=AgentOwner, PlayerOwner:=player[AgentOwner]):
            if:
                PickaxePerksUpgradesUI := MaybePickaxePerksUI?
            then:
                PickaxePerksUpgradesUI.Display_Canvas()
                PickaxePerksUpgradesUI.UpdateUI(PlayerDataCache)
        else:
            GameManager.Display_NotOwner(Agent)

    Display_DepositsUpgrades(Agent : agent) : void =
        if(AgentOwner:=MaybeAgentOwner?, Agent=AgentOwner, PlayerOwner:=player[AgentOwner], DepositsUpgradeUI := MaybeDepositsUpgradeUI?):
            if:
                PickaxeUpgradesUI := MaybePickaxeUpgradesUI?
            then:
                DepositsUpgradeUI.Display_Canvas()
                DepositsUpgradeUI.UpdateUI(PlayerDataCache)
        else:
            GameManager.Display_NotOwner(Agent)
    

    Display_ComputerBBInfo(Agent : agent) : void =
        if(AgentOwner:=MaybeAgentOwner?, Agent=AgentOwner, BBakersUpgradesUI := MaybeBBakersUpgradesUI?, PlayerOwner:=player[AgentOwner]):
            BBakersUpgradesUI.Display_Canvas()
            BBakersUpgradesUI.UpdateUI(PlayerDataCache,BB_Production)
        else:
            GameManager.Display_NotOwner(Agent)

    Display_ComputerCBBInfo(Agent : agent) : void =
        if(AgentOwner:=MaybeAgentOwner?, Agent=AgentOwner, CBBakersUpgradesUI := MaybeCBBakersUpgradesUI?, PlayerOwner:=player[AgentOwner]):
            if(IsCBBAvailable?):
                CBBakersUpgradesUI.Display_Canvas()
                CBBakersUpgradesUI.UpdateUI(PlayerDataCache,CBB_Production)
            else:
                GameManager.DisplayCharacterMessage(Agent,I_CHARACTER,"You have not unlocked this Baker",?Sound:=ERROR_SOUND)
        else:
            GameManager.Display_NotOwner(Agent)

    Display_ComputerPCSBInfo(Agent : agent) : void =
        #Print("Trying to display PCSB Info")
        if(AgentOwner:=MaybeAgentOwner?, Agent=AgentOwner, PCSBakersUpgradesUI := MaybePCSBakersUpgradesUI?, PlayerOwner:=player[AgentOwner]):
            if(IsCBBAvailable?):
                PCSBakersUpgradesUI.Display_Canvas()
                PCSBakersUpgradesUI.UpdateUI(PlayerDataCache,PCSB_Production)
            else:
                GameManager.DisplayCharacterMessage(Agent,I_CHARACTER,"You have not unlocked this Baker",?Sound:=ERROR_SOUND)
        else:
            GameManager.Display_NotOwner(Agent)

    Display_DroneUpgradesInfo(Agent : agent) : void =
        if(AgentOwner:=MaybeAgentOwner?, Agent=AgentOwner, DroneUpgradesUI := MaybeDroneUpgrades?, PlayerOwner:=player[AgentOwner]):
            DroneUpgradesUI.Display_Canvas()
            DroneUpgradesUI.UpdateUI(PlayerDataCache)
        else:
            GameManager.Display_NotOwner(Agent)

    Display_BackpackUpgrades(Agent : agent) : void =
        if(AgentOwner:=MaybeAgentOwner?, Agent=AgentOwner, BackpackUpgradesUI := MaybeBackpackUpgradesUI?, PlayerOwner:=player[AgentOwner]):
            BackpackUpgradesUI.Display_Canvas()
            BackpackUpgradesUI.UpdateUI(PlayerDataCache)
        else:
            GameManager.Display_NotOwner(Agent)

    #Info bakers
    Display_InfoBB(Agent : agent) : void =
        if(AgentOwner:=MaybeAgentOwner?, Agent=AgentOwner, BBInfoUI := MaybeBBInfoUI?, PlayerOwner:=player[AgentOwner]):
            set BBInfoIsDisplayed = true
            AllEfficiency := array{BB_WaterEfficiency,BB_FlourEfficiency}
            Log("For Display_InfoB Req_WaterFor_B: {Req_WaterFor_B}",LOG_LEVEL_UIs)
            Log("For Display_InfoB Req_FlourFor_B: {Req_FlourFor_B}",LOG_LEVEL_UIs)
            #SET BASIC INFO
            if(not IsOnUKTutorial?):
                BBInfoUI.SetResourceSlider(PlayerDataCache.WaterPreparedFor_B,Req_WaterFor_B,WATER)
                BBInfoUI.SetResourceSlider(PlayerDataCache.FlourPreparedFor_B,Req_FlourFor_B,FLOUR)
                BBInfoUI.SetProgressForB(Progress_ForB,B_BachProduction*SuperEnergy_Multiplier)
                BBInfoUI.Display_Canvas(AllEfficiency, BB_Production, CalculeBakerSpeed(BB_TYPE, PlayerDataCache.LV_BBSpeed, 1)(0))
            else if(UKTutorial:=MaybeUKTutorial?):
                BBInfoUI.SetResourceSlider(UKTutorial.WaterPreparedFor_B,UKTutorial.WaterReqFor_B,WATER)
                BBInfoUI.SetResourceSlider(UKTutorial.FlourPreparedFor_B,UKTutorial.FlourReqFor_B,FLOUR)
                BBInfoUI.SetProgressForB(0.0,1.0)
                BBInfoUI.Display_Canvas(array{1.0,1.0}, 1.0, 3.0)
            #Set progres
            #DISPLAY INFO
           
           
    Display_InfoCBB(Agent : agent) : void =
        if(AgentOwner:=MaybeAgentOwner?, Agent=AgentOwner, CBBInfoUI := MaybeCBBInfoUI?, PlayerOwner:=player[AgentOwner]):
            if(IsCBBAvailable?):
                set CBBInfoIsDisplayed = true
                AllEfficiency := array{CBB_WaterEfficiency,CBB_FlourEfficiency,CBB_CheeseEfficiency}
                Log("For Display_InfoCB Req_WaterFor_CB: {Req_WaterFor_CB}",LOG_LEVEL_UIs)
                #SET BASIC INFO
                CBBInfoUI.SetResourceSlider(PlayerDataCache.WaterPreparedFor_CB,Req_WaterFor_CB,WATER)
                CBBInfoUI.SetResourceSlider(PlayerDataCache.FlourPreparedFor_CB,Req_FlourFor_CB,FLOUR)
                CBBInfoUI.SetResourceSlider(PlayerDataCache.CheesePreparedFor_CB,Req_CheeseFor_CB,CHEESE)
                #Set progress
                CBBInfoUI.SetProgressForCBB(Progress_ForCB,CB_BachProduction*SuperEnergy_Multiplier)
                #DISPLAY INFO
                CBBInfoUI.Display_Canvas(AllEfficiency, CBB_Production, CalculeBakerSpeed(CBB_TYPE, PlayerDataCache.LV_CBBSpeed, 1)(0))
            else:
                GameManager.DisplayCharacterMessage(Agent,I_CHARACTER,"You have not unlocked this Baker",?Sound:=ERROR_SOUND)

    Display_InfoPCSB(Agent : agent) : void =
        if(AgentOwner:=MaybeAgentOwner?, Agent=AgentOwner, PCSBInfoUI := MaybePCSBInfoUI?, PlayerOwner:=player[AgentOwner]):
            if(IsPCSBAvailable?):
                set PCSBInfoIsDisplayed = true
                AllEfficiency := array{PCSB_WaterEfficiency,PCSB_FlourEfficiency,PCSB_CheeseEfficiency,PCSB_PimentosEfficiency}
                Log("For Display_InfoPCS Req_WaterFor_PCS: {Req_WaterFor_PCS}",LOG_LEVEL_UIs)
                #SET BASIC INFO
                PCSBInfoUI.SetResourceSlider(PlayerDataCache.WaterPreparedFor_PCS,Req_WaterFor_PCS,WATER)
                PCSBInfoUI.SetResourceSlider(PlayerDataCache.FlourPreparedFor_PCS,Req_FlourFor_PCS,FLOUR)
                PCSBInfoUI.SetResourceSlider(PlayerDataCache.CheesePreparedFor_PCS,Req_CheeseFor_PCS,CHEESE)
                PCSBInfoUI.SetResourceSlider(PlayerDataCache.PimentosPreparedFor_PCS,Req_CheeseFor_PCS,PIMENTOS)
                #Set progress
                PCSBInfoUI.SetProgressForPCS(Progress_ForPCS,PCS_BachProduction*SuperEnergy_Multiplier)
                #DISPLAY INFO
                PCSBInfoUI.Display_Canvas(AllEfficiency, PCSB_Production, CalculeBakerSpeed(PCSB_TYPE, PlayerDataCache.LV_PCSBSpeed, 1)(0))
            else:
                GameManager.DisplayCharacterMessage(Agent,I_CHARACTER,"You have not unlocked this Baker",?Sound:=ERROR_SOUND)
            

    UpdateValveProps(Index : int, Value : float) : void =
        if(Prop := AllProps_Valves[Index]):
            if(Value = 0.0):
                Prop.SetMaterial(Mat_PipeState0, ?Index:=0)
            if(Value = 1.0):
                Prop.SetMaterial(Mat_PipeState1, ?Index:=0)
            if(Value = 2.0):
                Prop.SetMaterial(Mat_PipeState2, ?Index:=0)
        else:
            #Print("FATAL ERROR on UpdateValveProps(): AllProps_Valves[{Index}] not found!")

    UpdateSubdeposits() : void =
        if:
            WaterPer_B_ValveProp := AllProps_Valves[0]
            FlourPer_B_ValveProp := AllProps_Valves[1]

            WaterPer_CB_ValveProp := AllProps_Valves[2]
            FlourPer_CB_ValveProp := AllProps_Valves[3]
            CheesePer_CB_ValveProp := AllProps_Valves[4]

            WaterPer_PCS_ValveProp := AllProps_Valves[5]
            FlourPer_PCS_ValveProp := AllProps_Valves[6]
            CheesePer_PCS_ValveProp := AllProps_Valves[7]
            PimentosPer_PCS_ValveProp := AllProps_Valves[8]
        then:
            Log("Req_WaterFor_B: {GetOptimizedMagnitudeLetter(Req_WaterFor_B)}",LOG_LEVEL_DEBUG8)
            Log("Req_FlourFor_B: {GetOptimizedMagnitudeLetter(Req_FlourFor_B)}",LOG_LEVEL_DEBUG8)
            Log("Req_WaterFor_CB: {GetOptimizedMagnitudeLetter(Req_WaterFor_CB)}",LOG_LEVEL_DEBUG8)
            Log("Req_FlourFor_CB: {GetOptimizedMagnitudeLetter(Req_FlourFor_CB)}",LOG_LEVEL_DEBUG8)
            Log("Req_CheeseFor_CB: {GetOptimizedMagnitudeLetter(Req_CheeseFor_CB)}",LOG_LEVEL_DEBUG8)
            Log("Req_WaterFor_PCS: {GetOptimizedMagnitudeLetter(Req_WaterFor_PCS)}",LOG_LEVEL_DEBUG8)
            Log("Req_FlourFor_PCS: {GetOptimizedMagnitudeLetter(Req_FlourFor_PCS)}",LOG_LEVEL_DEBUG8)
            Log("Req_CheeseFor_PCS: {GetOptimizedMagnitudeLetter(Req_CheeseFor_PCS)}",LOG_LEVEL_DEBUG8)
            Log("Req_PimentosFor_PCS: {GetOptimizedMagnitudeLetter(Req_PimentosFor_PCS)}",LOG_LEVEL_DEBUG8)
            if:
                Percentage_WaterPer_B := Floor[(PlayerDataCache.WaterPreparedFor_B/Req_WaterFor_B)*4]
                Percentage_FlourPer_B := Floor[(PlayerDataCache.FlourPreparedFor_B/Req_FlourFor_B)*4]
                Percentage_WaterPer_CB := Floor[(PlayerDataCache.WaterPreparedFor_CB/Req_WaterFor_CB)*4]
                Percentage_FlourPer_CB := Floor[(PlayerDataCache.FlourPreparedFor_CB/Req_FlourFor_CB)*4]
                Percentage_CheesePer_CB := Floor[(PlayerDataCache.CheesePreparedFor_CB/Req_CheeseFor_CB)*4]
                Percentage_WaterPer_PCS := Floor[(PlayerDataCache.WaterPreparedFor_PCS/Req_WaterFor_PCS)*4]
                Percentage_FlourPer_PCS := Floor[(PlayerDataCache.FlourPreparedFor_PCS/Req_FlourFor_PCS)*4]
                Percentage_CheesePer_PCS := Floor[(PlayerDataCache.CheesePreparedFor_PCS/Req_CheeseFor_PCS)*4]
                Percentage_PimentosPer_PCS := Floor[(PlayerDataCache.PimentosPreparedFor_PCS/Req_PimentosFor_PCS)*4]
            then:
                for(Index:=2..5):
                    #Bread
                    if(Percentage_WaterPer_B+1  >= Index):
                        WaterPer_B_ValveProp.SetMaterial(Mat_GreenScreen, ?Index:=Index)
                    else:
                        WaterPer_B_ValveProp.SetMaterial(Mat_RedScreen, ?Index:=Index)
    
                    if(Percentage_FlourPer_B+1  >= Index):
                        FlourPer_B_ValveProp.SetMaterial(Mat_GreenScreen, ?Index:=Index)
                    else:
                        FlourPer_B_ValveProp.SetMaterial(Mat_RedScreen, ?Index:=Index)
                    #Cheese Bread
                    if(Percentage_WaterPer_CB+1  >= Index):
                        WaterPer_CB_ValveProp.SetMaterial(Mat_GreenScreen, ?Index:=Index)
                    else:
                        WaterPer_CB_ValveProp.SetMaterial(Mat_RedScreen, ?Index:=Index)
                    if(Percentage_FlourPer_CB+1  >= Index):
                        FlourPer_CB_ValveProp.SetMaterial(Mat_GreenScreen, ?Index:=Index)
                    else:
                        FlourPer_CB_ValveProp.SetMaterial(Mat_RedScreen, ?Index:=Index)
                    if(Percentage_CheesePer_CB+1  >= Index):
                        CheesePer_CB_ValveProp.SetMaterial(Mat_GreenScreen, ?Index:=Index)
                    else:
                        CheesePer_CB_ValveProp.SetMaterial(Mat_RedScreen, ?Index:=Index)
                    #Pimento Cheese Bread
                    if(Percentage_WaterPer_PCS+1  >= Index):
                        WaterPer_PCS_ValveProp.SetMaterial(Mat_GreenScreen, ?Index:=Index)
                    else:
                        WaterPer_PCS_ValveProp.SetMaterial(Mat_RedScreen, ?Index:=Index)
                    if(Percentage_FlourPer_PCS+1  >= Index):
                        FlourPer_PCS_ValveProp.SetMaterial(Mat_GreenScreen, ?Index:=Index)
                    else:
                        FlourPer_PCS_ValveProp.SetMaterial(Mat_RedScreen, ?Index:=Index)
                    if(Percentage_CheesePer_PCS+1  >= Index):
                        CheesePer_PCS_ValveProp.SetMaterial(Mat_GreenScreen, ?Index:=Index)
                    else:
                        CheesePer_PCS_ValveProp.SetMaterial(Mat_RedScreen, ?Index:=Index)
                    if(Percentage_PimentosPer_PCS+1  >= Index):
                        PimentosPer_PCS_ValveProp.SetMaterial(Mat_GreenScreen, ?Index:=Index)
                    else:
                        PimentosPer_PCS_ValveProp.SetMaterial(Mat_RedScreen, ?Index:=Index)
            else:
                PrintFatalError("On UpdateSubdeposits(): percentage Failed")
        else:
            PrintFatalError("On UpdateSubdeposits(): Deposit Valve Prop not found")


    Calculate_ResTaken(MainDeposit : float, ValveState : float, TotalValvesValue : float, AmmountReq_Resource : float, CurrentAmmount_Resource : float) : float =
        var AmmountToTake : float = 0.0
        if(AmmountReq_Resource>0.0):
                set AmmountToTake = (MainDeposit/TotalValvesValue) * ValveState
                if(AmmountToTake > (AmmountReq_Resource - CurrentAmmount_Resource)):
                    set AmmountToTake = AmmountReq_Resource - CurrentAmmount_Resource
        return AmmountToTake

    UpdateAllMainDeposits() : void =
        if:
            MainWaterDeposit_Limit := AllMainDeposit_Limits[WATER]
            MainFlourDeposit_Limit := AllMainDeposit_Limits[FLOUR]
            MainCheeseDeposit_Limit := AllMainDeposit_Limits[CHEESE]
            MainPimentosDeposit_Limit := AllMainDeposit_Limits[PIMENTOS]
        then:

            WaterDeposit_Billbord.SetText(BasicMessage("{GetOptimizedMagnitudeLetter(PlayerDataCache.MainWaterDeposit_Value)}\n―――\n{GetOptimizedMagnitudeLetter(MainWaterDeposit_Limit)}"))
            FlourDeposit_Billbord.SetText(BasicMessage("{GetOptimizedMagnitudeLetter(PlayerDataCache.MainFlourDeposit_Value)}\n―――\n{GetOptimizedMagnitudeLetter(MainFlourDeposit_Limit)}"))
            CheeseDeposit_Billbord.SetText(BasicMessage("{GetOptimizedMagnitudeLetter(PlayerDataCache.MainCheeseDeposit_Value)}\n―――\n{GetOptimizedMagnitudeLetter(MainCheeseDeposit_Limit)}"))
            PimentosDeposit_Billbord.SetText(BasicMessage("{GetOptimizedMagnitudeLetter(PlayerDataCache.MainPimentosDeposit_Value)}\n―――\n{GetOptimizedMagnitudeLetter(MainPimentosDeposit_Limit)}"))

            spawn:
                AnimateDeposit(MainWaterDeposit_AmmountProp,MainWaterDeposit_Limit,PlayerDataCache.MainWaterDeposit_Value, false)
            spawn:
                AnimateDeposit(MainFlourDeposit_AmmountProp,MainFlourDeposit_Limit,PlayerDataCache.MainFlourDeposit_Value, false)
            spawn:
                AnimateDeposit(MainCheeseDeposit_AmmountProp,MainCheeseDeposit_Limit, PlayerDataCache.MainCheeseDeposit_Value, false)
            spawn:
                AnimateDeposit(MainPimentosDeposit_AmmountProp,MainPimentosDeposit_Limit,PlayerDataCache.MainPimentosDeposit_Value, false) 

        if(Agent := MaybeAgentOwner?, Player := player[Agent]):
            if(PCSBInfoIsDisplayed? or CBBInfoIsDisplayed? or BBInfoIsDisplayed?):
                if(PCSBInfoIsDisplayed?, PCSBInfoUI := MaybePCSBInfoUI?):
                    PCSBInfoUI.SetResourceSlider(PlayerDataCache.WaterPreparedFor_PCS,Req_WaterFor_PCS,WATER)
                    PCSBInfoUI.SetResourceSlider(PlayerDataCache.FlourPreparedFor_PCS,Req_FlourFor_PCS,FLOUR)
                    PCSBInfoUI.SetResourceSlider(PlayerDataCache.CheesePreparedFor_PCS,Req_CheeseFor_PCS,CHEESE)
                    PCSBInfoUI.SetResourceSlider(PlayerDataCache.PimentosPreparedFor_PCS,Req_PimentosFor_PCS,PIMENTOS)
                    PCSBInfoUI.SetProductionInfo(PCSB_Production, CalculeBakerSpeed(PCSB_TYPE, PlayerDataCache.LV_PCSBSpeed, 1)(0))

                if(CBBInfoIsDisplayed?, CBBInfoUI := MaybeCBBInfoUI?):
                    CBBInfoUI.SetResourceSlider(PlayerDataCache.WaterPreparedFor_CB,Req_WaterFor_CB,WATER)
                    CBBInfoUI.SetResourceSlider(PlayerDataCache.FlourPreparedFor_CB,Req_FlourFor_CB,FLOUR)
                    CBBInfoUI.SetResourceSlider(PlayerDataCache.CheesePreparedFor_CB,Req_CheeseFor_CB,CHEESE)
                    CBBInfoUI.SetProductionInfo(CBB_Production, CalculeBakerSpeed(CBB_TYPE, PlayerDataCache.LV_CBBSpeed, 1)(0))

                if(BBInfoIsDisplayed?, BBInfoUI := MaybeBBInfoUI?, not IsOnUKTutorial?):
                    BBInfoUI.SetResourceSlider(PlayerDataCache.WaterPreparedFor_B,Req_WaterFor_B,WATER)
                    BBInfoUI.SetResourceSlider(PlayerDataCache.FlourPreparedFor_B,Req_FlourFor_B,FLOUR)
                    BBInfoUI.SetProductionInfo(BB_Production, CalculeBakerSpeed(BB_TYPE, PlayerDataCache.LV_BBSpeed, 1)(0))

            #var NewData : player_data_table = PlayersDataManager.SetDepositResourcers(PlayerData,(MainWaterDeposit_Value,MainFlourDeposit_Value,MainCheeseDeposit_Value,MainPimentosDeposit_Value))
            #set NewData = PlayersDataManager.SetSubDepositResourcers(NewData,(WaterPreparedFor_B,FlourPreparedFor_B,WaterPreparedFor_CB,FlourPreparedFor_CB,CheesePreparedFor_CB,WaterPreparedFor_PCS,FlourPreparedFor_PCS,CheesePreparedFor_PCS,PimentosPreparedFor_PCS))
            #PlayersDataManager.SaveNewData(Player,PlayerData,NewData)

    UpdateDepositCinematic<private>(DepositCinematic : cinematic_sequence_device, Value : float, MaxValue : float) : void =
        Ammount := Value / MaxValue
        #Log("Updating deposit Value: {Value}", LOG_LEVEL_BAKERS)
        DepositCinematic.Play()
        DepositCinematic.SetPlaybackTime(Ammount)
        DepositCinematic.Pause()

    OnPlayerRemoved() : void =
        set IsBakerReady = false
        PetsManager.OnPlayerRemoved()
        if(DungeonUI:=MaybeDungeonUI?):
            DungeonUI.OnPlayerRemoved()
        HexaGameManager.OnPlayerRemoved()
        Reset()
        #Close all UIS
        CloseAllUI()

    Reset() : void =

        #Super Energy
        #set SuperEnergy_Owned = 0.0
        SuperEnergyRemaining_Timer.SetActiveDuration(0.0)
        SuperEnergyRemaining_Timer.SetMaxDuration(0.0) 

        set BBInfoIsDisplayed = false
        set CBBInfoIsDisplayed = false 
        set PCSBInfoIsDisplayed = false 

        set AllBackpackLimits = array{0.0,0.0,0.0,0.0}

        set AllAgentsMutantMaterialBackpackLimit = 0.0
        set AllAgentsCoalBackpackLimit = 0.0

        #Pickaxe stats
        set PickaxeTotalStrength = 0.0
        set AllPickaxeEfficiency = array{0.0,0.0,0.0,0.0,0.0}

        set PickaxeCriticalProb = 0.0
        set PickaxeCriticalDamage = 0.0
        set PickaxeCoalLuck = 0.0
        set PickaxeWeakPointMult = 0.0
        set PickaxeMultiResource = 0

        #Currencies
        #Rewards per products
        #set SapphiresPerB = Base_Req_WaterFor_B+Base_Req_FlourFor_B
        #set SapphiresPerCB = Base_Req_WaterFor_CB+Base_Req_FlourFor_CB+Base_Req_CheeseFor_CB
        #set SapphiresPerPCS = Base_Req_WaterFor_PCS+Base_Req_FlourFor_PCS+Base_Req_CheeseFor_PCS+Base_Req_PimentosFor_PCS

        #Drones
        #Drone stats
        set LV_DroneSpeed = 0
        set LV_DroneStorage = 0

        #set MaxB_Storage = Base_B_Storage
        #set MaxCB_Storage = Base_CB_Storage
        #set MaxPCS_Storage = Base_PCS_Storage

        set AllLV_SalesEfficiency = array{0,0,0}
        set All_SalesEfficiency = array{0.0,0.0,0.0}

        set DroneSpeed = Base_Drone_Speed
        set TimeToDroneReturn= 0
        set IsDronePropOnBase = true

    #Auto Pipes
        #set AllLV_PasiveResourcers = array{0,0,0,0}
        set AllPasiveResourcers = array{0.0,0.0,0.0,0.0}

    #Main deposits
        set AllLV_MainDeposits = array{0,0,0,0}
        set AllMainDeposit_Limits = array{1000.0,1000.0,1000.0,1000.0}

        set MainMutantMaterialDeposit_Value = 0.0
    #Baker Machines
            #Props
            #Progress
        set Progress_ForB = 0.0
        set Progress_ForCB = 0.0
        set Progress_ForPCS = 0.0
            #Level of production
        #set TLV_BB = array{0,-1,-1,-1,-1,-1,-1,-1,-1,-1,-1,-1,-1,-1,-1,-1,-1,-1,-1,-1}
        #set TLV_CBB = array{0,-1,-1,-1,-1,-1,-1,-1,-1,-1,-1,-1,-1,-1,-1,-1,-1,-1,-1,-1}
        #set TLV_PCSB = array{0,-1,-1,-1,-1,-1,-1,-1,-1,-1,-1,-1,-1,-1,-1,-1,-1,-1,-1,-1}
            #Ammount produced
        set BB_Production = 0.0
        set CBB_Production = 0.0
        set PCSB_Production = 0.0
            #Speed
        set BB_Speed = Base_BB_Speed
        set CBB_Speed = Base_CBB_Speed
        set PCSB_Speed = Base_PCSB_Speed
            #Efficiency
        set BB_WaterEfficiency = 2.5
        set CBB_WaterEfficiency = 2.5
        set PCSB_WaterEfficiency  = 2.5
        set BB_FlourEfficiency  = 2.5
        set CBB_FlourEfficiency  = 2.5
        set PCSB_FlourEfficiency  = 2.5
        set BB_CheeseEfficiency  = 2.5
        set CBB_CheeseEfficiency  = 2.5
        set PCSB_CheeseEfficiency  = 2.5
        set PCSB_PimentosEfficiency  = 2.5

    #Subdeposits
            #Requirements
        set Req_WaterFor_B = 999.0
        set Req_FlourFor_B = 999.0
        set Req_WaterFor_CB = 999.0
        set Req_FlourFor_CB = 999.0
        set Req_CheeseFor_CB = 999.0
        set Req_WaterFor_PCS = 999.0
        set Req_FlourFor_PCS = 999.0
        set Req_CheeseFor_PCS = 999.0
        set Req_PimentosFor_PCS = 999.0
        set Max_PimentosFor_Pets = 999.0
            #AmmountValues

        #Valves
        #States
        set ValveState_WaterFor_B = 0.0
        set ValveState_FlourFor_B = 0.0
        UpdateValveProps(0, 0.0)
        UpdateValveProps(1, 0.0)

        #if(LV_CheeseBreadBaker>=0):
        set ValveState_WaterFor_CB = 0.0
        set ValveState_FlourFor_CB = 0.0
        set ValveState_CheeseFor_CB = 0.0
        UpdateValveProps(2, 0.0)
        UpdateValveProps(3, 0.0)
        UpdateValveProps(4, 0.0)

        #if(LV_PimentoCheeseBlender>=0):
        set ValveState_WaterFor_PCS = 0.0
        set ValveState_FlourFor_PCS = 0.0
        set ValveState_CheeseFor_PCS = 0.0
        set ValveState_PimentosFor_PCS = 0.0
        UpdateValveProps(6, 0.0)
        UpdateValveProps(5, 0.0)
        UpdateValveProps(7, 0.0)
        #UpdateValveProps(8, 0.0)

        #set ValveState_PimentosFor_Pets = 0.0
        
        #set ValveState_PimentosFor_Pet = 0
        
        # Flows
        set WaterAmmount_ForB_ToTake = 0.0
        set WaterAmmount_ForCB_ToTake = 0.0
        set WaterAmmount_ForPCS_ToTake = 0.0
        set FlourAmmount_ForB_ToTake = 0.0
        set FlourAmmount_ForCB_ToTake = 0.0
        set FlourAmmount_ForPCS_ToTake = 0.0
        set CheeseAmmount_ForCB_ToTake = 0.0
        set CheeseAmmount_ForPCS_ToTake = 0.0
        set PimentosAmmount_ForPCS_ToTake = 0.0
        #set PimentosAmmount_ForPets_ToTake = 0.0
        
        # Flow Per Second
        set WaterFlowPerSecond_ForB_ToTake = 0.0
        set WaterFlowPerSecond_ForCB_ToTake = 0.0
        set WaterFlowPerSecond_ForPCS_ToTake = 0.0
        set FlourFlowPerSecond_ForB_ToTake = 0.0
        set FlourFlowPerSecond_ForCB_ToTake = 0.0
        set FlourFlowPerSecond_ForPCS_ToTake = 0.0
        set CheeseFlowPerSecond_ForCB_ToTake = 0.0
        set CheeseFlowPerSecond_ForPCS_ToTake = 0.0
        set PimentosFlowPerSecond_ForPCS_ToTake = 0.0
        #set PimentosFlowPerSecond_ForPets_ToTake = 0.0
            
        #String and Messages
        #Strings
        set VALVEFLOW_BUTTON_WATER_MESSAGE = "💧 Flow: "
        set VALVEFLOW_BUTTON_FLOUR_MESSAGE = "🌾 Flow: "
        set VALVEFLOW_BUTTON_CHEESE_MESSAGE = "🧀 Flow: "
        set VALVEFLOW_BUTTON_PIMENTOS_MESSAGE = "🌶️ Flow: "
        
    #UPGRADE
    DisplayBasicUIs() : void =
        if:
            Agent:=MaybeAgentOwner?
            ActionInfoUI:=MaybeActionInfoUI?
            MainInventoryUI:=MaybeMainInventoryUI?
            EntityHealthUI:=MaybeEntityHealthUI?
        then:
            ActionInfoUI.Display_Canvas()
            MainInventoryUI.Display_Canvas()
            EntityHealthUI.Display_Canvas()
        else:
            PrintFatalError("On DisplayBasicUIs")
        RefreshAllUI()


    #Deposit
    TryUpgradeDeposit_Capacity(Row : int) : void =
        Log("Upgrading deposit capacity {Row}",LOG_LEVEL_UIs)
        if:
            DepositsUpgradeUI:=MaybeDepositsUpgradeUI?
            Agent:=MaybeAgentOwner?
            PlayerOwner:=player[Agent]
        then:
            set AllLV_MainDeposits = PlayerDataCache.AllLV_MainDeposits
            if(Level := AllLV_MainDeposits[Row]):
                TupleLimit:=CalculeDepositUpgradeLimit(Level,Row)
              
                CurrentLimit:=TupleLimit(0)
                NextCost:=TupleLimit(1)
                Difference:=TupleLimit(2)

                TupleHasEnough:=PlayersDataManager.HasEnoughResource(PlayerDataCache,NextCost,Row)
                if(TupleHasEnough(0)>NOT_RES_STATE):
                    GameManager.VFXManager.SmallObjetive_AudioPlayer.Play(Agent)
                    if(TupleHasEnough(0)=SECOND_RES_STATE):
                        Log("Resource got from SecondayDeposits {GetOptimizedMagnitudeLetter(TupleHasEnough(2))}",LOG_LEVEL_DEBUG4)
                        SetNewPlayerDataCache(PlayersDataManager.WithdrawResourceFromSecondaryDeposit(PlayerDataCache,TupleHasEnough(2),Row))
                        SetNewPlayerDataCache(PlayersDataManager.WithdrawResourceFromMainDeposit(PlayerDataCache,NextCost,Row))
                        SetNewPlayerDataCache(PlayersDataManager.WithdrawResource(PlayerDataCache,NextCost,Row))
                    else if(TupleHasEnough(0)=MAIN_RES_STATE):
                        Log("Resource got from PrimaryDeposits {GetOptimizedMagnitudeLetter(TupleHasEnough(1))}",LOG_LEVEL_DEBUG4)
                        SetNewPlayerDataCache(PlayersDataManager.WithdrawResourceFromMainDeposit(PlayerDataCache,TupleHasEnough(1),Row))
                        SetNewPlayerDataCache(PlayersDataManager.WithdrawResource(PlayerDataCache,NextCost,Row))
                    else if(TupleHasEnough(0)=BACKPACK_RES_STATE):
                        Log("Resource got from Backpack",LOG_LEVEL_DEBUG4)
                        SetNewPlayerDataCache(PlayersDataManager.WithdrawResource(PlayerDataCache,NextCost,Row))
                    if(set AllLV_MainDeposits[Row]=Level+1):
                    #SAVE NEW DATA
                    SetNewPlayerDataCache(PlayersDataManager.SetLV_MainDeposits(PlayerDataCache,AllLV_MainDeposits))
                    #PlayersDataManager.SaveAllNewData(PlayerOwner, PlayerData, NewTableCompleted, Self)
                    SetAllMainDepositLimits(Agent,AllLV_MainDeposits)
                    DepositsUpgradeUI.UpdateUI(PlayerDataCache)
                    RefreshInventoryUI(PlayerDataCache)
                    CheckQuestProgress(1.0,IMPROVEMAINDEPOSITS_TUTORIAL_TYPE)
                    #ACCOLADE
                    GameManager.VFXManager.AwardAccolade(BAKER_ACCOLADE,Agent)
                else:
                    spawn:
                        GameManager.VFXManager.PlaySound(ERROR_SOUND,0.0,option{Agent},false)
            else:
                spawn:
                    GameManager.VFXManager.PlaySound(ERROR_SOUND,0.0,option{Agent},false)

    #PASIVE
    TryUpgradeDeposit_PasiveIncome(Row : int) : void =
        
        if:
            DepositsUpgradeUI:=MaybeDepositsUpgradeUI?
            Agent:=MaybeAgentOwner?
            PlayerOwner:=player[Agent]
            #PlayerData := PlayersDataManager.GetPlayerData[PlayerOwner]
        then:
            var AllLV_PasiveResourcers : []int = PlayerDataCache.AllLV_PasiveResourcers
            if(Level := PlayerDataCache.AllLV_PasiveResourcers[Row]):
                
                TuplePasive:=CalculeDepositPasiveIncome(Level)
              
                CurrentPasive:=TuplePasive(0)
                NextCost:=TuplePasive(1)
                Difference:=TuplePasive(2)

                if(PlayerDataCache.Inventory_AgedSapphires >= NextCost, Level < MaxUpgradePasiveIncome):
                    Log("Upgrading Pasive Income {Row}",LOG_LEVEL_DEBUG5)
                    
                    if(set AllLV_PasiveResourcers[Row]=Level+1):
                    else:
                        PrintFatalError("On TryUpgradeDeposit_PasiveIncome, Row: {Row} not found!")
                    #SAVE NEW DATA
                    NewTableWithoutAgedSapphires:=PlayersDataManager.WithdrawResource(PlayerDataCache,NextCost,AGEDSAPPHIRES)
                    SetNewPlayerDataCache(PlayersDataManager.SetLV_PasiveResourcers(NewTableWithoutAgedSapphires,AllLV_PasiveResourcers))
                    #PlayersDataManager.SaveAllNewData(PlayerOwner, PlayerDataCache, NewTableCompleted, Self)
                    SetAllPasiveResourcers()
                    DepositsUpgradeUI.UpdateUI(PlayerDataCache)
                    RefreshInventoryUI(PlayerDataCache)
                    #ACCOLADE
                    GameManager.VFXManager.AwardAccolade(BAKER_ACCOLADE,Agent)
                else:
                    spawn:
                        GameManager.VFXManager.PlaySound(ERROR_SOUND,0.0,option{Agent},false)
            else:
                spawn:
                    GameManager.VFXManager.PlaySound(ERROR_SOUND,0.0,option{Agent},false)

            set AllLV_PasiveResourcers = PlayerDataCache.AllLV_PasiveResourcers

    CheckProductionReq(Production : float, BakerType : int) : tuple(logic,int) =
        var NumMaxPicks : int = 30
        var NumPicks : int = 30
        var IsExceded : logic = false
        #Pickaxe current damage
        TuplePickDamage := GetBakeryPickaxeDamage(PlayerDataCache.LV_Area+1,false,true)
        var TotalDamage : float = TuplePickDamage(0)
        var ResourcersAmmount : float = ApplyEfficiencyToDamage(TotalDamage)
        if:
            WaterPerPick := <#ResourcersAmmount + #>AllPickaxeEfficiency[WATER]*ResourcersAmmount
            FlourPerPick := <#ResourcersAmmount + #>AllPickaxeEfficiency[FLOUR]*ResourcersAmmount  
            CheesePerPick := <#ResourcersAmmount + #>AllPickaxeEfficiency[FLOUR]*ResourcersAmmount
            PimentosPerPick := <#ResourcersAmmount + #>AllPickaxeEfficiency[PIMENTOS]*ResourcersAmmount
        then:
            #Num Max Hits
            if(PlayerDataCache.LV_Area > 7):
                set NumMaxPicks = 70
            else if(PlayerDataCache.LV_Area > 5):
                set NumMaxPicks = 50
            else if(PlayerDataCache.LV_Area > 3):
                set NumMaxPicks = 40
            #Log("NumMaxPicks: {NumMaxPicks}",LOG_LEVEL_DEBUG5)
            #BB
            if(BakerType = BB_TYPE): 
                CalculatedReq_WaterFor_B := (Base_Req_WaterFor_B * Production)/BB_WaterEfficiency
                CalculatedReq_FlourFor_B := (Base_Req_FlourFor_B * Production)/BB_FlourEfficiency

                WaterNumPicks := CalculatedReq_WaterFor_B/WaterPerPick
                FlourNumPicks := CalculatedReq_FlourFor_B/FlourPerPick

                if(WaterNumPicks>NumMaxPicks*1.0, WaterNumPicks>NumPicks*1.0, IntWaterPicks := Ceil[WaterNumPicks]):
                    set NumPicks = IntWaterPicks
                    set IsExceded = true
                if(FlourNumPicks>NumMaxPicks*1.0, FlourNumPicks>NumPicks*1.0, IntFlourPicks := Ceil[FlourNumPicks]):
                    set NumPicks = IntFlourPicks
                    set IsExceded = true
            #CBB
            if(BakerType = CBB_TYPE): 
                CalculatedReq_WaterFor_CB := (Base_Req_WaterFor_CB * Production)/CBB_WaterEfficiency
                CalculatedReq_FlourFor_CB := (Base_Req_FlourFor_CB * Production)/CBB_FlourEfficiency
                CalculatedReq_CheeseFor_CB := (Base_Req_CheeseFor_CB * Production)/CBB_CheeseEfficiency

                WaterNumPicks := CalculatedReq_WaterFor_CB/WaterPerPick
                FlourNumPicks := CalculatedReq_FlourFor_CB/FlourPerPick
                CheeseNumPicks := CalculatedReq_CheeseFor_CB/CheesePerPick

                if(WaterNumPicks>NumMaxPicks*1.0, WaterNumPicks>NumPicks*1.0, IntWaterPicks := Ceil[WaterNumPicks]):
                    set NumPicks = IntWaterPicks
                    set IsExceded = true
                if(FlourNumPicks>NumMaxPicks*1.0, FlourNumPicks>NumPicks*1.0, IntFlourPicks := Ceil[FlourNumPicks]):
                    set NumPicks = IntFlourPicks
                    set IsExceded = true
                if(CheeseNumPicks>NumMaxPicks*1.0, CheeseNumPicks>NumPicks*1.0, IntCheesePicks := Ceil[CheeseNumPicks]):
                    set NumPicks = IntCheesePicks
                    set IsExceded = true
            #PCSB
            if(BakerType = PCSB_TYPE): 
                CalculatedReq_WaterFor_PCS := (Base_Req_WaterFor_PCS * Production)/PCSB_WaterEfficiency
                CalculatedReq_FlourFor_PCS := (Base_Req_FlourFor_PCS * Production)/PCSB_FlourEfficiency
                CalculatedReq_CheeseFor_PCS := (Base_Req_CheeseFor_PCS * Production)/PCSB_CheeseEfficiency
                CalculatedReq_PimentosFor_PCS := (Base_Req_PimentosFor_PCS * Production)/PCSB_PimentosEfficiency

                WaterNumPicks := CalculatedReq_WaterFor_PCS/WaterPerPick
                FlourNumPicks := CalculatedReq_FlourFor_PCS/FlourPerPick
                CheeseNumPicks := CalculatedReq_CheeseFor_PCS/CheesePerPick
                PimentosNumPicks := CalculatedReq_PimentosFor_PCS/PimentosPerPick

                if(WaterNumPicks>NumMaxPicks*1.0, WaterNumPicks>NumPicks*1.0, IntWaterPicks := Ceil[WaterNumPicks]):
                    set NumPicks = IntWaterPicks
                    set IsExceded = true
                if(FlourNumPicks>NumMaxPicks*1.0, FlourNumPicks>NumPicks*1.0, IntFlourPicks := Ceil[FlourNumPicks]):
                    set NumPicks = IntFlourPicks
                    set IsExceded = true
                if(CheeseNumPicks>NumMaxPicks*1.0, CheeseNumPicks>NumPicks*1.0, IntCheesePicks := Ceil[CheeseNumPicks]):
                    set NumPicks = IntCheesePicks
                    set IsExceded = true
                if(PimentosNumPicks>NumMaxPicks*1.0, PimentosNumPicks>NumPicks*1.0, IntPimentosPicks := Ceil[PimentosNumPicks]):
                    set NumPicks = IntPimentosPicks
                    set IsExceded = true

        return (IsExceded,NumPicks)

    #Production
    TryUpgradeBaker_Production(Row : int, BakerType : int) : void =
        if:
            BBakersUpgradesUI:=MaybeBBakersUpgradesUI?
            CBBakersUpgradesUI:=MaybeCBBakersUpgradesUI?
            PCSBakersUpgradesUI:=MaybePCSBakersUpgradesUI?
            Agent:=MaybeAgentOwner?
            PlayerOwner:=player[Agent]
            #PlayerData := PlayersDataManager.GetPlayerData[PlayerOwner]
        then:
            var TLV_Baker : []int = array{} 
            var Tier : int = 0
            var BuyModeMultiplier : int = 0
            var Level : int = 0
            
            if(BakerType=BB_TYPE):
                set TLV_Baker = PlayerDataCache.TLV_BB
                set Tier = (BBakersUpgradesUI.CurrentPage-1)*5+Row
                set BuyModeMultiplier = BBakersUpgradesUI.BuyModeMultiplier
                if(LevelAux := PlayerDataCache.TLV_BB[Tier], set Level = LevelAux):

            if(BakerType=CBB_TYPE):
                set TLV_Baker = PlayerDataCache.TLV_CBB
                set Tier = (CBBakersUpgradesUI.CurrentPage-1)*5+Row
                set BuyModeMultiplier = CBBakersUpgradesUI.BuyModeMultiplier
                if(LevelAux := PlayerDataCache.TLV_CBB[Tier], set Level = LevelAux):

            if(BakerType=PCSB_TYPE):
                set TLV_Baker = PlayerDataCache.TLV_PCSB
                set Tier = (PCSBakersUpgradesUI.CurrentPage-1)*5+Row
                set BuyModeMultiplier = PCSBakersUpgradesUI.BuyModeMultiplier
                if(LevelAux := PlayerDataCache.TLV_PCSB[Tier], set Level = LevelAux):
            if(not IsOnUKTutorial?):
                if:
                    Level<>-1, Level<MaxTierLevel
                then:
                    TupleProduction:=CalculeBakerUpgradeProduction(Tier,Level,BakerType,BuyModeMultiplier)
                    if:
                        NextCost:=TupleProduction(1)
                        NextProduction:=TupleProduction(3)
                        PlayerDataCache.Inventory_Sapphires >= NextCost
                        NewTableWithoutSapphires:=PlayersDataManager.SetSapphires(PlayerDataCache,PlayerDataCache.Inventory_Sapphires-NextCost)
                    then:
                        TupleProductionCheck := CheckProductionReq(NextProduction,BakerType)
                        NumHitsRequired := TupleProductionCheck(1)
                        if(NumHitsRequired<250):
                            if(not TupleProductionCheck(0)? or WarningMessageDisplayed?):
                                set WarningMessageDisplayed = false
                                if(set TLV_Baker[Tier] = Level+BuyModeMultiplier):
                                if(TLV_Baker[Tier+1]=-1, set TLV_Baker[Tier+1] = 0):
                                if(TLV_Baker[Tier]>MaxTierLevel, set TLV_Baker[Tier] = MaxTierLevel): #Check if max level was reached
                                #SAVE NEW DATA
                                GameManager.VFXManager.AwardAccolade(BAKER_ACCOLADE,Agent)
                                spawn:
                                    GameManager.VFXManager.PlaySound(SMALLOBJETIVE_SOUND,0.0,option{Agent},false)
                                if(BakerType=BB_TYPE):
                                    #set TLV_BB = TLV_Baker
                                    NewTableCompleted:=PlayersDataManager.SetTLV_BBaker(NewTableWithoutSapphires,TLV_Baker)
                                    SetNewPlayerDataCache(NewTableCompleted)
                                    #PlayersDataManager.SaveAllNewData(PlayerOwner, PlayerData, NewTableCompleted, Self)
                                    set BB_Production = CalculeTotalProduction(PlayerDataCache.TLV_BB,BB_TYPE)
                                    BBakersUpgradesUI.UpdateUI(NewTableCompleted,BB_Production)
                                    SetBakerEfficiency(NewTableCompleted,BB_TYPE)
                                    RefreshInventoryUI(NewTableCompleted)

                                if(BakerType=CBB_TYPE):
                                    #set TLV_CBB = TLV_Baker
                                    NewTableCompleted:=PlayersDataManager.SetTLV_CBBaker(NewTableWithoutSapphires,TLV_Baker)
                                    SetNewPlayerDataCache(NewTableCompleted)
                                    #PlayersDataManager.SaveAllNewData(PlayerOwner, PlayerData, NewTableCompleted, Self)
                                    set CBB_Production = CalculeTotalProduction(PlayerDataCache.TLV_CBB,CBB_TYPE)
                                    CBBakersUpgradesUI.UpdateUI(NewTableCompleted,CBB_Production)
                                    SetBakerEfficiency(NewTableCompleted,CBB_TYPE)
                                    RefreshInventoryUI(NewTableCompleted)
                        
                                if(BakerType=PCSB_TYPE):
                                    #set TLV_PCSB = TLV_Baker
                                    NewTableCompleted:=PlayersDataManager.SetTLV_PCSBaker(NewTableWithoutSapphires,TLV_Baker)
                                    SetNewPlayerDataCache(NewTableCompleted)
                                    #PlayersDataManager.SaveAllNewData(PlayerOwner, PlayerData, NewTableCompleted, Self)
                                    set PCSB_Production = CalculeTotalProduction(PlayerDataCache.TLV_PCSB,PCSB_TYPE)
                                    PCSBakersUpgradesUI.UpdateUI(NewTableCompleted,PCSB_Production)
                                    SetBakerEfficiency(NewTableCompleted,PCSB_TYPE)
                                    RefreshInventoryUI(NewTableCompleted)
                                SetDroneMaxStorages()
                                CheckQuestProgress(1.0,IMPROVEBAKERY_TUTORIAL_TYPE)
                            else:
                                spawn:
                                    GameManager.VFXManager.PlaySound(WARNING_SOUND,0.0,option{Agent},false)
                                if(BakerType=BB_TYPE):
                                    set WarningMessageDisplayed = true
                                    spawn:
                                        BBakersUpgradesUI.RunWarning(TupleProductionCheck(1))
                                if(BakerType=CBB_TYPE):
                                    set WarningMessageDisplayed = true
                                    spawn:
                                        CBBakersUpgradesUI.RunWarning(TupleProductionCheck(1))
                                if(BakerType=PCSB_TYPE):
                                    set WarningMessageDisplayed = true
                                    spawn:
                                        PCSBakersUpgradesUI.RunWarning(TupleProductionCheck(1))
                        else:
                            BBakersUpgradesUI.Remove_CanvasWithoutMessage()
                            CBBakersUpgradesUI.Remove_CanvasWithoutMessage()
                            PCSBakersUpgradesUI.Remove_CanvasWithoutMessage()
                            spawn:
                                GameManager.VFXManager.PlaySound(WARNING_SOUND,0.0,option{Agent},false)
                            GameManager.DisplayCharacterMessage(Agent,I_CHARACTER,"Upgrading production requires x{NumHitsRequired} pickaxe hits!",?Sound:=WARNING_SOUND,?DisplayTime:=5.0)
                            GameManager.DisplayCharacterMessage(Agent,I_CHARACTER,"Improve your pickaxe damage first!",?Sound:=WARNING_SOUND,?DisplayTime:=5.0)
                    else:
                        spawn:
                            GameManager.VFXManager.PlaySound(ERROR_SOUND,0.0,option{Agent},false)
                else:
                    spawn:
                        GameManager.VFXManager.PlaySound(ERROR_SOUND,0.0,option{Agent},false)
            else:
                if(Tier=0, UKTutorial := MaybeUKTutorial?):
                    UKTutorial.BakerUpgradedEvent()
                    BBakersUpgradesUI.Remove_CanvasWithoutMessage()
                else:
                    spawn:
                        GameManager.VFXManager.PlaySound(ERROR_SOUND,0.0,option{Agent},false)

    #Efficiency
    TryUpgradeBaker_Efficiency(Row : int, BakerType : int) : void =
        if:
            BBakersUpgradesUI:=MaybeBBakersUpgradesUI?
            CBBakersUpgradesUI:=MaybeCBBakersUpgradesUI?
            PCSBakersUpgradesUI:=MaybePCSBakersUpgradesUI?
            Agent:=MaybeAgentOwner?
            PlayerOwner:=player[Agent]
            #PlayerData := PlayersDataManager.GetPlayerData[PlayerOwner]
        then:
            var ArrayEfficiency_LVs : []int = array{}
            var BuyModeMultiplier : int = 0
            var ResEfficiencyLevel : int = 0
            
            if(BakerType=BB_TYPE):
                set ArrayEfficiency_LVs = PlayerDataCache.EfficiencyLV_BB 
                set BuyModeMultiplier = BBakersUpgradesUI.BuyModeMultiplier

            if(BakerType=CBB_TYPE):
                set ArrayEfficiency_LVs = PlayerDataCache.EfficiencyLV_CBB 
                set BuyModeMultiplier = CBBakersUpgradesUI.BuyModeMultiplier

            if(BakerType=PCSB_TYPE):
                set ArrayEfficiency_LVs = PlayerDataCache.EfficiencyLV_PCSB         
                set BuyModeMultiplier = PCSBakersUpgradesUI.BuyModeMultiplier

            if:
                AuxLevel := ArrayEfficiency_LVs[Row]
                set ResEfficiencyLevel = AuxLevel
                ResEfficiencyLevel < MaxBakerEfficiencyLV
            then:
                TupleEfficiency:=CalculeBakerResourceEfficiency(ResEfficiencyLevel,BakerType,Row,BuyModeMultiplier)
                var CurrentEfficiency : float = TupleEfficiency(0)
                NextCost:=TupleEfficiency(1)
                Difference:=TupleEfficiency(2)
                if:
                    PlayerDataCache.Inventory_Sapphires >= NextCost,
                    NewTableWithoutSapphires:=PlayersDataManager.SetSapphires(PlayerDataCache,PlayerDataCache.Inventory_Sapphires-NextCost)
                then:
                    var NewLevel : int = ResEfficiencyLevel + BuyModeMultiplier
                    if(NewLevel>MaxBakerEfficiencyLV):
                        set NewLevel = MaxBakerEfficiencyLV
                    if(set ArrayEfficiency_LVs[Row] = NewLevel):
                            #SAVE NEW DATA
                        if(BakerType=BB_TYPE):
                            if(Row=WATER):
                                set BB_WaterEfficiency = 1.0+(CalculeBakerResourceEfficiency(NewLevel, B_TYPE, WATER, 1)(0)/100.0)
                            else if(Row=FLOUR):
                                set BB_FlourEfficiency = 1.0+(CalculeBakerResourceEfficiency(NewLevel, B_TYPE, FLOUR, 1)(0)/100.0)

                            NewTableCompleted:=PlayersDataManager.SetBB_Efficiency(NewTableWithoutSapphires,ArrayEfficiency_LVs)
                            SetNewPlayerDataCache(NewTableCompleted)
                            #PlayersDataManager.SaveAllNewData(PlayerOwner, PlayerData, NewTableCompleted, Self)
                            GameManager.VFXManager.AwardAccolade(BAKER_ACCOLADE,Agent)
                            spawn:
                                GameManager.VFXManager.PlaySound(SMALLOBJETIVE_SOUND,0.0,option{Agent},false)
                            set BB_Production = CalculeTotalProduction(PlayerDataCache.TLV_BB,BB_TYPE)
                            BBakersUpgradesUI.UpdateUI(NewTableCompleted,BB_Production)
                            SetBakerEfficiency(NewTableCompleted,BB_TYPE)
                            RefreshInventoryUI(NewTableCompleted)

                        if(BakerType=CBB_TYPE):
                            if(Row=WATER):
                                set CBB_WaterEfficiency = 1.0+(CalculeBakerResourceEfficiency(NewLevel, CB_TYPE, WATER, 1)(0)/100.0)
                            else if(Row=FLOUR):
                                set CBB_FlourEfficiency = 1.0+(CalculeBakerResourceEfficiency(NewLevel, CB_TYPE, FLOUR, 1)(0)/100.0)
                            else if(Row=CHEESE):
                                set CBB_CheeseEfficiency = 1.0+(CalculeBakerResourceEfficiency(NewLevel, CB_TYPE, CHEESE, 1)(0)/100.0)

                            NewTableCompleted:=PlayersDataManager.SetCBB_Efficiency(NewTableWithoutSapphires,ArrayEfficiency_LVs)
                            SetNewPlayerDataCache(NewTableCompleted)
                            #PlayersDataManager.SaveAllNewData(PlayerOwner, PlayerData, NewTableCompleted, Self)
                            GameManager.VFXManager.AwardAccolade(BAKER_ACCOLADE,Agent)
                            spawn:
                                GameManager.VFXManager.PlaySound(SMALLOBJETIVE_SOUND,0.0,option{Agent},false)
                            set CBB_Production = CalculeTotalProduction(PlayerDataCache.TLV_CBB,CBB_TYPE)
                            CBBakersUpgradesUI.UpdateUI(NewTableCompleted,CBB_Production)
                            SetBakerEfficiency(NewTableCompleted,CBB_TYPE)
                            RefreshInventoryUI(NewTableCompleted)

                        if(BakerType=PCSB_TYPE):
                            if(Row=WATER):
                                set PCSB_WaterEfficiency = 1.0+(CalculeBakerResourceEfficiency(NewLevel, PCSB_TYPE, WATER, 1)(0)/100.0)
                            else if(Row=FLOUR):
                                set PCSB_FlourEfficiency = 1.0+(CalculeBakerResourceEfficiency(NewLevel, PCSB_TYPE, FLOUR, 1)(0)/100.0)
                            else if(Row=CHEESE):
                                set PCSB_CheeseEfficiency = 1.0+(CalculeBakerResourceEfficiency(NewLevel, PCSB_TYPE, CHEESE, 1)(0)/100.0)
                            else if(Row=PIMENTOS):
                                set PCSB_PimentosEfficiency = 1.0+(CalculeBakerResourceEfficiency(NewLevel, PCSB_TYPE, PIMENTOS, 1)(0)/100.0)

                            NewTableCompleted:=PlayersDataManager.SetPCSB_Efficiency(NewTableWithoutSapphires,ArrayEfficiency_LVs)
                            SetNewPlayerDataCache(NewTableCompleted)
                            #PlayersDataManager.SaveAllNewData(PlayerOwner, PlayerData, NewTableCompleted, Self)
                            GameManager.VFXManager.AwardAccolade(BAKER_ACCOLADE,Agent)
                            spawn:
                                GameManager.VFXManager.PlaySound(SMALLOBJETIVE_SOUND,0.0,option{Agent},false)
                            set PCSB_Production = CalculeTotalProduction(PlayerDataCache.TLV_PCSB,PCSB_TYPE)
                            PCSBakersUpgradesUI.UpdateUI(NewTableCompleted,PCSB_Production)
                            SetBakerEfficiency(NewTableCompleted,PCSB_TYPE)
                            RefreshInventoryUI(NewTableCompleted)
                else:
                    spawn:
                        GameManager.VFXManager.PlaySound(ERROR_SOUND,0.0,option{Agent},false)
            else:
                spawn:
                    GameManager.VFXManager.PlaySound(ERROR_SOUND,0.0,option{Agent},false)

    #Baker speed
    TryUpgradeBaker_Speed(Row : int, BakerType : int) : void =
        if:
            var MaxSpeedLV : int = 0
            BBakersUpgradesUI:=MaybeBBakersUpgradesUI?
            CBBakersUpgradesUI:=MaybeCBBakersUpgradesUI?
            PCSBakersUpgradesUI:=MaybePCSBakersUpgradesUI?
            Agent:=MaybeAgentOwner?
            PlayerOwner:=player[Agent]
            #PlayerData := PlayersDataManager.GetPlayerData[PlayerOwner]
            var LV_Speed : int = 0
        then:
            var ArrayEfficiency_LVs : []int = array{}
            var BuyModeMultiplier : int = 0
            var ResEfficiencyLevel : int = 0
            
            if(BakerType=BB_TYPE):
                set BuyModeMultiplier = BBakersUpgradesUI.BuyModeMultiplier
                set MaxSpeedLV = CalculeSpeedLimitLevel(Base_BB_Speed)
                set LV_Speed = PlayerDataCache.LV_BBSpeed

            if(BakerType=CBB_TYPE):
                set BuyModeMultiplier = CBBakersUpgradesUI.BuyModeMultiplier
                set  MaxSpeedLV = CalculeSpeedLimitLevel(Base_CBB_Speed)
                set LV_Speed = PlayerDataCache.LV_CBBSpeed

            if(BakerType=PCSB_TYPE):
                set BuyModeMultiplier = PCSBakersUpgradesUI.BuyModeMultiplier
                set  MaxSpeedLV = CalculeSpeedLimitLevel(Base_PCSB_Speed)         
                set LV_Speed = PlayerDataCache.LV_PCSBSpeed
            if:
                LV_Speed<MaxSpeedLV
            then:
                TupleSpeed:=CalculeBakerSpeed(BakerType,LV_Speed,BuyModeMultiplier)
                    var CurrentSpeed : float = TupleSpeed(0)
                    NextCost:=TupleSpeed(1)
                    Difference:=TupleSpeed(2)
                    if:
                        PlayerDataCache.Inventory_Sapphires >= NextCost,
                        NewTableWithoutSapphires:=PlayersDataManager.SetSapphires(PlayerDataCache,PlayerDataCache.Inventory_Sapphires-NextCost)
                    then:
                        var NewLevel : int = LV_Speed + BuyModeMultiplier    
                        if(NewLevel>MaxSpeedLV):
                            set NewLevel = MaxSpeedLV
                        set LV_Speed = NewLevel
                        #SAVE NEW DATA
                        if(BakerType=BB_TYPE):
                            NewTableCompleted:=PlayersDataManager.SetBB_Speed(NewTableWithoutSapphires,LV_Speed)
                            SetNewPlayerDataCache(NewTableCompleted)
                            #PlayersDataManager.SaveAllNewData(PlayerOwner, PlayerData, NewTableCompleted, Self)
                            set BB_Production = CalculeTotalProduction(PlayerDataCache.TLV_BB,BB_TYPE)
                            BBakersUpgradesUI.UpdateUI(NewTableCompleted,BB_Production)
                            set BB_Speed = CalculeBakerSpeed(BB_TYPE,LV_Speed,BuyModeMultiplier)(0)
                            RefreshInventoryUI(NewTableCompleted)
                        if(BakerType=CBB_TYPE):
                            NewTableCompleted:=PlayersDataManager.SetCBB_Speed(NewTableWithoutSapphires,LV_Speed)
                            SetNewPlayerDataCache(NewTableCompleted)
                            #PlayersDataManager.SaveAllNewData(PlayerOwner, PlayerData, NewTableCompleted, Self)
                            set CBB_Production = CalculeTotalProduction(PlayerDataCache.TLV_CBB,CBB_TYPE)
                            CBBakersUpgradesUI.UpdateUI(NewTableCompleted,CBB_Production)
                            set CBB_Speed = CalculeBakerSpeed(CBB_TYPE,LV_Speed,BuyModeMultiplier)(0)
                            RefreshInventoryUI(NewTableCompleted)
                        if(BakerType=PCSB_TYPE):
                            NewTableCompleted:=PlayersDataManager.SetPCSB_Speed(NewTableWithoutSapphires,LV_Speed)
                            SetNewPlayerDataCache(NewTableCompleted)
                            #PlayersDataManager.SaveAllNewData(PlayerOwner, PlayerData, NewTableCompleted, Self)
                            set PCSB_Production = CalculeTotalProduction(PlayerDataCache.TLV_PCSB,PCSB_TYPE)
                            PCSBakersUpgradesUI.UpdateUI(NewTableCompleted,PCSB_Production)
                            set PCSB_Speed = CalculeBakerSpeed(PCSB_TYPE,LV_Speed,BuyModeMultiplier)(0)
                            RefreshInventoryUI(NewTableCompleted)
                        GameManager.VFXManager.AwardAccolade(BAKER_ACCOLADE,Agent)
                        spawn:
                            GameManager.VFXManager.PlaySound(SMALLOBJETIVE_SOUND,0.0,option{Agent},false)
                        SetDroneMaxStorages()
                    else:
                        spawn:
                            GameManager.VFXManager.PlaySound(ERROR_SOUND,0.0,option{Agent},false)
            else:
                spawn:
                    GameManager.VFXManager.PlaySound(ERROR_SOUND,0.0,option{Agent},false)

    #Drone efficiency
    TrySalesEfficiencyUpgrade(ProductType : int) : void =
        Log("Updating sales efficiency upgrade {ProductType} ", LOG_LEVEL_DEBUG7)
        if:
            DroneUpgrades:=MaybeDroneUpgrades?
            Agent:=MaybeAgentOwner?
            PlayerOwner:=player[Agent]
            #PlayerData := PlayersDataManager.GetPlayerData[PlayerOwner]
            set AllLV_SalesEfficiency = PlayerDataCache.AllLV_SalesEfficiency
            Product_LVSalesEfficiency := AllLV_SalesEfficiency[ProductType]
        then:
            if(Product_LVSalesEfficiency<MaxSalesLV):
                TupleSalesEfficency:=CalculeSalesEfficiency(Product_LVSalesEfficiency)
                
                CurrentSales:=TupleSalesEfficency(0)
                NextCost:=TupleSalesEfficency(1)
                Difference:=TupleSalesEfficency(2)

                if(PlayerDataCache.Inventory_AgedSapphires >= NextCost):
                    NewTableWithoutSapphires:=PlayersDataManager.SetAgedSapphires(PlayerDataCache,PlayerDataCache.Inventory_AgedSapphires-NextCost)
                    if(Value := AllLV_SalesEfficiency[ProductType]+1, set AllLV_SalesEfficiency[ProductType]=Value):
                        #SAVE NEW DATA
                        SetNewPlayerDataCache(PlayersDataManager.SetSalesEfficiency(NewTableWithoutSapphires,AllLV_SalesEfficiency))
                        #PlayersDataManager.SaveAllNewData(PlayerOwner, PlayerData, NewTableCompleted, Self)
                        SetSalesEfficiency(PlayerDataCache,ProductType)
                        Log("New level: {Value}", LOG_LEVEL_DEBUG7)
                        DroneUpgrades.UpdateUI(PlayerDataCache)
                        GameManager.VFXManager.AwardAccolade(AGEDSAPPHIRES_ACCOLADE,Agent)
                else:
                    spawn:
                        GameManager.VFXManager.PlaySound(ERROR_SOUND,0.0,option{Agent},false)
            else:
                spawn:
                    GameManager.VFXManager.PlaySound(ERROR_SOUND,0.0,option{Agent},false)


    #Drone speed
    TryUpgradeDroneSpeed() : void =
        Log("Updating drone speed", LOG_LEVEL_UIs)
        if:
            DroneUpgrades:=MaybeDroneUpgrades?
            Agent:=MaybeAgentOwner?
            PlayerOwner:=player[Agent]
            #PlayerData := PlayersDataManager.GetPlayerData[PlayerOwner]
        then:
           set LV_DroneSpeed = PlayerDataCache.LV_DroneSpeed

            MaxDroneSpeedLV := CalculeDroneSpeedLimitLevel()

            if(LV_DroneSpeed<MaxDroneSpeedLV):

                TuplePasive:=CalculeDroneSpeed(LV_DroneSpeed)
                
                CurrentSpeed:=TuplePasive(0)
                NextCost:=TuplePasive(1)
                Difference:=TuplePasive(2)

                if(PlayerDataCache.Inventory_Sapphires >= NextCost):
                    NewTableWithoutSapphires:=PlayersDataManager.SetSapphires(PlayerDataCache,PlayerDataCache.Inventory_Sapphires-NextCost)
                    set LV_DroneSpeed=LV_DroneSpeed+1
                    #SAVE NEW DATA
                    NewTableCompleted:=PlayersDataManager.SetDroneSpeed(NewTableWithoutSapphires,LV_DroneSpeed)
                    if(TimeToDroneReturn>0):
                        set TimeToDroneReturn=RecalculeNewDroneSpeed(DroneSpeed,CalculeDroneSpeed(LV_DroneSpeed)(0),TimeToDroneReturn)
                        SetNewPlayerDataCache(NewTableCompleted)
                        #PlayersDataManager.SaveAllNewData(PlayerOwner, PlayerData, NewTableCompleted, Self)
                    SetDroneSettings(NewTableCompleted,false)
                    DroneUpgrades.UpdateUI(NewTableCompleted)
                    RefreshInventoryUI(NewTableCompleted)
                    GameManager.VFXManager.AwardAccolade(BAKER_ACCOLADE,Agent)
                else:
                    spawn:
                        GameManager.VFXManager.PlaySound(ERROR_SOUND,0.0,option{Agent},false)
            else:
                spawn:
                    GameManager.VFXManager.PlaySound(ERROR_SOUND,0.0,option{Agent},false)

    #BackPack
    TryBackpackUpgrade(Resource : int, ?AutoBackpackAmount : float = 0.0) : float =
        Log("Updating Backpack, resource {Resource}", LOG_LEVEL_DEBUG8)
        var AutoBackpackAmount_Remaining : float = 0.0
        if:
            BackpackUpgrades:=MaybeBackpackUpgradesUI?
            Agent:=MaybeAgentOwner?
            PlayerOwner:=player[Agent]
            #PlayerData := PlayersDataManager.GetPlayerData[PlayerOwner]
            var AllLV_BackpackLimits : []int = PlayerDataCache.AllLV_BackpackLimits
            ResourceBackpackLevel := AllLV_BackpackLimits[Resource]
        then:
         
            BackpackUpgrade := CalculeBackpackUpgrade(ResourceBackpackLevel,Resource)
                
            CurrentBackpack:=BackpackUpgrade(0)
            NextCost:=BackpackUpgrade(1)
            Difference:=BackpackUpgrade(2)
            Log("Backpack current: {CurrentBackpack}. NextCost: {NextCost}. Difference: {Difference}",LOG_LEVEL_UIs)
            TupleHasEnough:=PlayersDataManager.HasEnoughResource(PlayerDataCache,NextCost,Resource)
            if(TupleHasEnough(0)<>NOT_RES_STATE):
                set AutoBackpackAmount_Remaining = AutoBackpackAmount
                GameManager.VFXManager.SmallObjetive_AudioPlayer.Play(Agent)
                if(TupleHasEnough(0)=SECOND_RES_STATE):
                    Log("Resource got from SecondayDeposits {GetOptimizedMagnitudeLetter(TupleHasEnough(2))}",LOG_LEVEL_DEBUG8)
                    SetNewPlayerDataCache(PlayersDataManager.WithdrawResourceFromSecondaryDeposit(PlayerDataCache,TupleHasEnough(2),Resource))
                    SetNewPlayerDataCache(PlayersDataManager.WithdrawResourceFromMainDeposit(PlayerDataCache,NextCost,Resource))
                    SetNewPlayerDataCache(PlayersDataManager.WithdrawResource(PlayerDataCache,NextCost,Resource))
                else if(TupleHasEnough(0)=MAIN_RES_STATE):
                    Log("Resource got from PrimaryDeposits {GetOptimizedMagnitudeLetter(TupleHasEnough(1))}",LOG_LEVEL_DEBUG8)
                    SetNewPlayerDataCache(PlayersDataManager.WithdrawResourceFromMainDeposit(PlayerDataCache,TupleHasEnough(1),Resource))
                    SetNewPlayerDataCache(PlayersDataManager.WithdrawResource(PlayerDataCache,NextCost,Resource))
                else if(TupleHasEnough(0)=BACKPACK_RES_STATE):
                    Log("Resource got from Backpack",LOG_LEVEL_DEBUG8)
                    SetNewPlayerDataCache(PlayersDataManager.WithdrawResource(PlayerDataCache,NextCost,Resource))
                if(set AllLV_BackpackLimits[Resource] = ResourceBackpackLevel + 1):
                    Log("Backpack Upgraded on : {Resource}. New : {ResourceBackpackLevel + 1}",LOG_LEVEL_UIs)
                #SAVE NEW DATA
                SetNewPlayerDataCache(PlayersDataManager.SetAllLV_BackpackLimits(PlayerDataCache,AllLV_BackpackLimits))
                #PlayersDataManager.SaveAllNewData(PlayerOwner, PlayerData, NewTableCompleted, Self)
                GenerateAllBackPackLimits(PlayerDataCache)
                RefreshInventoryUI(PlayerDataCache)
                BackpackUpgrades.UpdateUI(PlayerDataCache)
                CheckQuestProgress(1.0,IMPROVEBACKPACK_TUTORIAL_TYPE)
                GameManager.VFXManager.AwardAccolade(INVENTORY_ACCOLADE,Agent)
            else if(AutoBackpackAmount>0.0):
                Log("AutoBackpackDetected on TryBackpackUpgrade... Initiating Upgrade",LOG_LEVEL_DEBUG7)
                if(AutoBackpackAmount >= NextCost):
                    Log("Failed, but AutoBackpackAmount: {GetOptimizedMagnitudeLetter(AutoBackpackAmount)} is enough. Upgrading",LOG_LEVEL_DEBUG8)
                    if(set AllLV_BackpackLimits[Resource] = ResourceBackpackLevel + 1):
                    set AutoBackpackAmount_Remaining=AutoBackpackAmount-NextCost
                    SetNewPlayerDataCache(PlayersDataManager.SetAllLV_BackpackLimits(PlayerDataCache,AllLV_BackpackLimits))
                else:
                    if(Resource=WATER, AutoBackpackAmount + PlayerDataCache.Inventory_Water >= NextCost):
                        Log("Failed, but AutoBackpackAmount + Water Resource on Player: {GetOptimizedMagnitudeLetter(AutoBackpackAmount+PlayerDataCache.Inventory_Water)} is enough. Upgrading",LOG_LEVEL_DEBUG8)
                        set AutoBackpackAmount_Remaining = 0.0
                        SetNewPlayerDataCache(PlayersDataManager.WithdrawResource(PlayerDataCache,NextCost-AutoBackpackAmount,Resource))
                        if(set AllLV_BackpackLimits[Resource] = ResourceBackpackLevel + 1):
                        SetNewPlayerDataCache(PlayersDataManager.SetAllLV_BackpackLimits(PlayerDataCache,AllLV_BackpackLimits))
                    else if(Resource=FLOUR, AutoBackpackAmount + PlayerDataCache.Inventory_Flour >= NextCost):
                        Log("Failed, but AutoBackpackAmount + Flour Resource on Player: {GetOptimizedMagnitudeLetter(AutoBackpackAmount+PlayerDataCache.Inventory_Flour)} is enough. Upgrading",LOG_LEVEL_DEBUG8)
                        SetNewPlayerDataCache(PlayersDataManager.WithdrawResource(PlayerDataCache,NextCost-AutoBackpackAmount,Resource))
                        if(set AllLV_BackpackLimits[Resource] = ResourceBackpackLevel + 1):
                        SetNewPlayerDataCache(PlayersDataManager.SetAllLV_BackpackLimits(PlayerDataCache,AllLV_BackpackLimits))
                    else if(Resource=CHEESE, AutoBackpackAmount + PlayerDataCache.Inventory_Cheese >= NextCost):
                        Log("Failed, but AutoBackpackAmount + Cheese Resource on Player: {GetOptimizedMagnitudeLetter(AutoBackpackAmount+PlayerDataCache.Inventory_Cheese)} is enough. Upgrading",LOG_LEVEL_DEBUG8)
                        SetNewPlayerDataCache(PlayersDataManager.WithdrawResource(PlayerDataCache,NextCost-AutoBackpackAmount,Resource))
                        if(set AllLV_BackpackLimits[Resource] = ResourceBackpackLevel + 1):
                        SetNewPlayerDataCache(PlayersDataManager.SetAllLV_BackpackLimits(PlayerDataCache,AllLV_BackpackLimits))
                    else if(Resource=PIMENTOS, AutoBackpackAmount + PlayerDataCache.Inventory_Pimentos >= NextCost):
                        Log("Failed, but AutoBackpackAmount + PimentosResource on Player: {GetOptimizedMagnitudeLetter(AutoBackpackAmount+PlayerDataCache.Inventory_Pimentos)} is enough. Upgrading",LOG_LEVEL_DEBUG8)
                        SetNewPlayerDataCache(PlayersDataManager.WithdrawResource(PlayerDataCache,NextCost-AutoBackpackAmount,Resource))
                        if(set AllLV_BackpackLimits[Resource] = ResourceBackpackLevel + 1):
                        SetNewPlayerDataCache(PlayersDataManager.SetAllLV_BackpackLimits(PlayerDataCache,AllLV_BackpackLimits))
                    else:
                        PrintWarning("Failed, no AutoBackpackAmount!!")
                spawn:
                    GameManager.VFXManager.PlaySound(ERROR_SOUND,0.0,option{Agent},false)

        return AutoBackpackAmount_Remaining

    #PickaxePerkUpgrade
    TryPickaxePerkUpgrade(Perk : int) : void =
        if:
            PickaxePerksUI:=MaybePickaxePerksUI?
            Agent:=MaybeAgentOwner?
            PlayerOwner:=player[Agent]
            #PlayerData := PlayersDataManager.GetPlayerData[PlayerOwner]
            var AllLV_AllLV_PickaxePerks : []int = PlayerDataCache.AllLV_PickaxePerks
            var PickaxePerkLV : int = AllLV_AllLV_PickaxePerks[Perk]
        then:
            var TupleResult : tuple(float,float,float) = (0.0,0.0,0.0)
            var MaxLevel : float = 99.0

            Log("Updating pickaxe perks {Perk}", LOG_LEVEL_UIs)

            if(Perk = CRITICAL_PROB):
                set TupleResult = CalculePickaxeCriticalProbUpgrade(PickaxePerkLV)
                set MaxLevel = GetMaxPickaxeCriticalProbLV()
            else if(Perk = CRITICAL_DAMAGE):
                set TupleResult = CalculePickaxeCriticalDamageUpgrade(PickaxePerkLV)
                set MaxLevel = GetMaxPickaxeCriticalDamageLV()
            else if(Perk = PICKAXE_LUCK):
                set TupleResult = CalculePickaxeLuckUpgrade(PickaxePerkLV)
                set MaxLevel = GetMaxLuckLV()
            else if(Perk = PICKAXE_WEAK_MULT):
                set TupleResult = CalculePickaxeWeakUpgrade(PickaxePerkLV)
                set MaxLevel = GetMaxWeakPointLV()
            else if(Perk = PICKAXE_MULT):
                set TupleResult = CalculeMultipleResUpgrade(PickaxePerkLV)
                set MaxLevel = MaxMultiResource
                
            CurrentBackpack:=TupleResult(0)
            NextCost:=TupleResult(1)
            Difference:=TupleResult(2)

            if((PickaxePerkLV*1.0)<MaxLevel):
                if(PlayerDataCache.Inventory_Sapphires >= NextCost):
                    NewTableWithoutSapphires:=PlayersDataManager.SetSapphires(PlayerDataCache,PlayerDataCache.Inventory_Sapphires-NextCost)
                    set PickaxePerkLV=PickaxePerkLV+1
                    if(set AllLV_AllLV_PickaxePerks[Perk] = PickaxePerkLV):
                    #SAVE NEW DATA
                    NewTableCompleted:=PlayersDataManager.SetPickaxePerks(NewTableWithoutSapphires,AllLV_AllLV_PickaxePerks)
                    SetNewPlayerDataCache(NewTableCompleted)
                    #PlayersDataManager.SaveAllNewData(PlayerOwner, PlayerData, NewTableCompleted, Self)
                    SetPickaxePerks(NewTableCompleted)
                    RefreshInventoryUI(NewTableCompleted)
                    PickaxePerksUI.UpdateUI(NewTableCompleted)
                    GameManager.VFXManager.AwardAccolade(PICKAXE_ACCOLADE,Agent)
                    spawn:
                        GameManager.VFXManager.PlaySound(SMALLOBJETIVE_SOUND,0.0,option{Agent},false)
                    CheckQuestProgress(1.0,IMPROVEPICKPERK_TUTORIAL_TYPE)
                else:
                    spawn:
                        GameManager.VFXManager.PlaySound(ERROR_SOUND,0.0,option{Agent},false)

    RefreshBProductionUI(BakerType : int) : void =
        case(BakerType):
            0=> 
                if(Agent:=MaybeAgentOwner?,PlayerOwner:=player[Agent],BBakersUpgradesUI:=MaybeBBakersUpgradesUI?):
                    BBakersUpgradesUI.UpdateUI(PlayerDataCache,BB_Production)
                else:
                    PrintFatalError("Fatal error updating UI BB")
            1=>
                if(Agent:=MaybeAgentOwner?,PlayerOwner:=player[Agent],CBBakersUpgradesUI:=MaybeCBBakersUpgradesUI?):
                    CBBakersUpgradesUI.UpdateUI(PlayerDataCache,CBB_Production)
                else:
                    PrintFatalError("Fatal error updating UI CBB")
            2=>
                if(Agent:=MaybeAgentOwner?,PlayerOwner:=player[Agent],PCSBakersUpgradesUI:=MaybePCSBakersUpgradesUI?):
                    PCSBakersUpgradesUI.UpdateUI(PlayerDataCache,PCSB_Production)
                else:
                    PrintFatalError("Fatal error updating UI PCSB")
            _=>#Print("FATAL ERROR Upgrading info. BakerType: {BakerType} not valid!")
            
    TrySendDrone(Agent : agent) : void =
        if(AgentOwner:=MaybeAgentOwner?, Player := player[Agent],Agent=AgentOwner, DroneUpgradesUI := MaybeDroneUpgrades?, PlayerOwner:=player[AgentOwner]):
            if(0.0<PlayerDataCache.B_On_Drone+PlayerDataCache.CB_On_Drone+PlayerDataCache.PCS_On_Drone):
                if(TimeToDroneReturn=0):
                    set IsDronePropOnBase = false
                    set TimeToDroneReturn = DroneSpeed
                    if:       
                        B_LVEfficiency:=PlayerDataCache.AllLV_SalesEfficiency[0]
                        CB_LVEfficiency:=PlayerDataCache.AllLV_SalesEfficiency[1]
                        PCS_LVEfficiency:=PlayerDataCache.AllLV_SalesEfficiency[2]
                        var NewPlayerData : player_data_table = PlayersDataManager.SetHistoric_ProductSent(PlayerDataCache,PlayerDataCache.Historic_BSent + PlayerDataCache.B_On_Drone, PlayerDataCache.Historic_CBSent + PlayerDataCache.CB_On_Drone, PlayerDataCache.Historic_PCSSent + PlayerDataCache.PCS_On_Drone)
                    then:
                        B_Efficiency := 1.0+(CalculeSalesEfficiency(B_LVEfficiency)(0)/100.0)
                        CB_Efficiency := 1.0+(CalculeSalesEfficiency(CB_LVEfficiency)(0)/100.0)
                        PCS_Efficiency := 1.0+(CalculeSalesEfficiency(PCS_LVEfficiency)(0)/100.0)

                        TotalSapphiresPer_B := B_Efficiency * SapphiresPerB * PlayerDataCache.B_On_Drone
                        TotalSapphiresPer_CB := CB_Efficiency * SapphiresPerCB * PlayerDataCache.CB_On_Drone
                        TotalSapphiresPer_PCS := PCS_Efficiency * SapphiresPerPCS * PlayerDataCache.PCS_On_Drone

                        TotalSapphiresToGrant := TotalSapphiresPer_B + TotalSapphiresPer_CB + TotalSapphiresPer_PCS

                        Log("B_Efficiency: {B_Efficiency}, SapphiresPerB: {SapphiresPerB}, B_On_Drone: {PlayerDataCache.B_On_Drone}",LOG_LEVEL_DEBUG7)
                        Log("TotalSapphiresToGrant: {TotalSapphiresToGrant}",LOG_LEVEL_DEBUG7)

                        #Quest
                        set NewPlayerData = GameManager.QuestsAndOrdersDatabase.AddAgentQuest_ProgressV2(Player,NewPlayerData,PlayerDataCache.B_On_Drone,B_TYPE)(4)
                        set NewPlayerData = GameManager.QuestsAndOrdersDatabase.AddAgentQuest_ProgressV2(Player,NewPlayerData,PlayerDataCache.CB_On_Drone,CB_TYPE)(4)
                        set NewPlayerData = GameManager.QuestsAndOrdersDatabase.AddAgentQuest_ProgressV2(Player,NewPlayerData,PlayerDataCache.PCS_On_Drone,PCS_TYPE)(4)
                        #Orders
                        set NewPlayerData = GameManager.QuestsAndOrdersDatabase.AddAgentOrder_Progress(Player,NewPlayerData,PlayerDataCache.B_On_Drone,PlayerDataCache.CB_On_Drone,PlayerDataCache.PCS_On_Drone)(4)

                        if(ActionInfoUI := MaybeActionInfoUI?):
                            ActionInfoUI.DisplayGather(array{TotalSapphiresToGrant}, array{SAPPHIRES}, array{false},false)

                        set NewPlayerData = PlayersDataManager.SetSapphires(NewPlayerData,PlayerDataCache.Inventory_Sapphires+TotalSapphiresToGrant)
                        set NewPlayerData = PlayersDataManager.ClearDroneAmmount(NewPlayerData)
                        
                        SetNewPlayerDataCache(NewPlayerData)
                        #PlayersDataManager.SaveAllNewData(Player, PlayerData, NewPlayerData, Self)

                        #RefreshInventoryUI(NewPlayerData)
                        RefreshAllUI()
                        UpdateDroneDepositBillBoard()
                        GameManager.VFXManager.AwardAccolade(DRONESEND_ACCOLADE,Agent)
                    spawn:
                        GameManager.VFXManager.AnimateDroneMovement(0, DroneProp, DroneInitialTransform, false)
                else:
                    GameManager.Display_DroneNotReady(Agent,TimeToDroneReturn)
            else:
                GameManager.DisplayCharacterMessage(Agent,W_CHARACTER,"The Drone's empty! We’re not wasting a trip for nothing!")
        else:
            GameManager.Display_NotOwner(Agent) 

    GetBakeryPickaxeDamage(HardBread_Level : int, IsWeakPoint : logic, IsAVG : logic) : tuple(float,logic) =
        var ResourcersAmmount : float = 0.0
        var TupleDamage : tuple(float,logic) = (0.0,false)
        if(not IsOnUKTutorial?):
            Pets_PickDamage_Multiplier := PetsManager.PlayerEnhacement_PickDamage
            Pets_WeakPointMult := PetsManager.PlayerEnhacement_WeakPointMult
            Pets_PickEfficency := PetsManager.PlayerEnhacement_PickEfficency
            Pet_PickCriticalDam := PetsManager.PlayerEnhacement_PickCriticalDam

            #TotalPickaxeTotalStrength:=PickaxeTotalStrength*Pets_PickDamage_Multiplier
            TotalPickaxeWeakPointMult:=PickaxeWeakPointMult+(Pets_WeakPointMult-1.0)
            TotalPickaxeCriticalDamage:=PickaxeCriticalDamage+(Pets_PickDamage_Multiplier-1.0)
            
            set TupleDamage = GetPickaxeDamage(IsAVG,PickaxeTotalStrength,IsWeakPoint,PickaxeCriticalProb,TotalPickaxeCriticalDamage,TotalPickaxeWeakPointMult,Pets_PickDamage_Multiplier)

            #HardBread Purity
            set ResourcersAmmount = TupleDamage(0) + TupleDamage(0)*HardBread_Tier_Multiplier*HardBread_Level
        else:
            if(UKTutorial:=MaybeUKTutorial?, not UKTutorial.IsPickUpgraded?):
                return (CalculeTotalStrength(player_data_table{}.TLV_PickaxeStrength),false)
            else:
                return (CalculeTotalStrength(array{1,-1,-1,-1,-1,-1,-1,-1,-1,-1,-1,-1,-1,-1,-1,-1,-1,-1,-1,-1}),false)



        return (ResourcersAmmount,TupleDamage(1))

            #GATHER
    
    ApplyEfficiencyToDamage(TotalDamage : float) : float =
        var ResourcersAmmount : float = TotalDamage
        ############## APPLYING MODIFIERS TO GATHER
        Pets_PickEfficency := PetsManager.PlayerEnhacement_PickEfficency
        #Secrets Efficiency
        set ResourcersAmmount = ResourcersAmmount + (TotalDamage*(Secret_PickaxeEfficiency-1.0))
        #Pets Efficiency
        set ResourcersAmmount = ResourcersAmmount + (TotalDamage*(Pets_PickEfficency-1.0))
        return ResourcersAmmount

    GetDamage_And_Resourcers(Agent : agent, NumHitsAccumulated : int, HardBread_Type : int, HardBread_Level : int, IsWeakPoint : logic, HealthRemaining : float, ?IsInestable : logic = false, ?InestableState : int = 0, ?AllPlayersDamage : [agent]float = map{}) : tuple(float,[]int,[]float,[]logic,logic) =
        <#
        Pets_PickDamage_Multiplier := PetsManager.PlayerEnhacement_PickDamage
        Pets_WeakPointMult := PetsManager.PlayerEnhacement_WeakPointMult
        
        Pet_PickCriticalDam := PetsManager.PlayerEnhacement_PickCriticalDam

        #TotalPickaxeTotalStrength:=PickaxeTotalStrength*Pets_PickDamage_Multiplier
        TotalPickaxeWeakPointMult:=PickaxeWeakPointMult+(Pets_WeakPointMult-1.0)
        TotalPickaxeCriticalDamage:=PickaxeCriticalDamage+(Pets_PickDamage_Multiplier-1.0)
        
        TuplePickDamage := GetPickaxeDamage(false,PickaxeTotalStrength,IsWeakPoint,PickaxeCriticalProb,TotalPickaxeCriticalDamage,TotalPickaxeWeakPointMult,Pets_PickDamage_Multiplier)
        #>
        TuplePickDamage := GetBakeryPickaxeDamage(HardBread_Level,IsWeakPoint,false)
        var TotalDamage : float = TuplePickDamage(0)
        var ResourcersAmmount : float = ApplyEfficiencyToDamage(TotalDamage)

        IsCritical := TuplePickDamage(1)
        var AllResourcersType : []int = array{}
        var AllResourcersAmmount : []float = array{}
        MultiResource := PickaxeMultiResource

        if(IsInestable?):
            Log("HB is Inestable, InestableState: {InestableState}", LOG_LEVEL_UIs)
                CheckQuestProgress(1.0,UNSTABLEHB_QUEST_TYPE)
            if(InestableState=1):
                set ResourcersAmmount = ResourcersAmmount * INSTABLE_BREAD_MULTIPLIER

        if(IsCritical?):
            spawn:
                GameManager.VFXManager.PlaySound(PICKCRITICAL_SOUND,0.0,option{Agent},false)

        ############## APPLYING RESOURCER MODIFIERS 
        var Water_BaseProb : float = 0.0
        var Flour_BaseProb : float = 0.0 
        var Cheese_BaseProb : float = 0.0
        var Pimentos_BaseProb : float = 0.0
        ############## APLYING MODIFIERS ON DESTROYED MODE    
        if(HardBread_Level<>0, HealthRemaining <= TotalDamage<#, NumHitsAccumulated*1.0 > MinNum_AVGHit#>):
            #Elimination detected
            CheckQuestProgress(1.0,BREAKBREAD_TUTORIAL_TYPE)
            GameManager.VFXManager.AwardAccolade(HARDBREAD_ACCOLADE,Agent)
            set ResourcersAmmount = ResourcersAmmount * 2.0
        else:
            IsThereCoal:=IsThereCoalAward()
            if(IsThereCoal?):
                Log("Coal gathered!", LOG_LEVEL_UIs)
                set AllResourcersType = AllResourcersType + array{COAL}
                set AllResourcersAmmount = AllResourcersAmmount + array{1.0}
        #var Mutant_BaseProb : float = 0.0
        if(HardBread_Type = HB_TYPE0):
            set Water_BaseProb = HB_TYPE0_Water_BaseProb
            set Flour_BaseProb = HB_TYPE0_Flour_BaseProb
            set Cheese_BaseProb = HB_TYPE0_Cheese_BaseProb
            set Pimentos_BaseProb = HB_TYPE0_Pimentos_BaseProb
            #set Mutant_BaseProb = HB_TYPE0_MutantMaterial_BaseProb
        else if(HardBread_Type = HB_TYPE1):
            set Water_BaseProb = HB_TYPE1_Water_BaseProb
            set Flour_BaseProb = HB_TYPE1_Flour_BaseProb
            set Cheese_BaseProb = HB_TYPE1_Cheese_BaseProb
            set Pimentos_BaseProb = HB_TYPE1_Pimentos_BaseProb
            #set Mutant_BaseProb = HB_TYPE1_MutantMaterial_BaseProb
        else if(HardBread_Type = HB_TYPE2):
            set Water_BaseProb = HB_TYPE2_Water_BaseProb
            set Flour_BaseProb = HB_TYPE2_Flour_BaseProb
            set Cheese_BaseProb = HB_TYPE2_Cheese_BaseProb
            set Pimentos_BaseProb = HB_TYPE2_Pimentos_BaseProb
            #set Mutant_BaseProb = HB_TYPE2_MutantMaterial_BaseProb
        else if(HardBread_Type = HB_TYPE3):
            set Water_BaseProb = HB_TYPE3_Water_BaseProb
            set Flour_BaseProb = HB_TYPE3_Flour_BaseProb
            set Cheese_BaseProb = HB_TYPE3_Cheese_BaseProb
            set Pimentos_BaseProb = HB_TYPE3_Pimentos_BaseProb
            #set Mutant_BaseProb = HB_TYPE3_MutantMaterial_BaseProb
        else if(HardBread_Type = HB_TYPE4 or HardBread_Type = HB_PURE):
            set Water_BaseProb = HB_TYPE4_Water_BaseProb
            set Flour_BaseProb = HB_TYPE4_Flour_BaseProb
            set Cheese_BaseProb = HB_TYPE4_Cheese_BaseProb
            set Pimentos_BaseProb = HB_TYPE4_Pimentos_BaseProb
        else if(HardBread_Type = HB_BASETYPE):
            set Water_BaseProb = HB_BASETYPE_Water_BaseProb
            set Flour_BaseProb = HB_BASETYPE_Flour_BaseProb
            set Cheese_BaseProb = HB_BASETYPE_Cheese_BaseProb
            set Pimentos_BaseProb = HB_BASETYPE_Pimentos_BaseProb

            #set Mutant_BaseProb = HB_TYPE4_MutantMaterial_BaseProb
        ############## APPLYING MODIFIERS TO DAMAGE
        for(NumResource:=1..MultiResource):
            var RandomResource : float = GetRandomFloat(0.0,1.0)
            if(NumResource=1, IsWeakPoint?, HB_TYPE4<>HardBread_Type, ResEfficiency := AllPickaxeEfficiency[HardBread_Type]):
                set AllResourcersType = AllResourcersType + array{HardBread_Type}
                set AllResourcersAmmount = AllResourcersAmmount + array{ResourcersAmmount*ResEfficiency}
            else if(NumResource=1, HardBread_Level<>0, HB_TYPE4<>HardBread_Type, HealthRemaining <= TotalDamage, ResEfficiency := AllPickaxeEfficiency[HardBread_Type]):
                set AllResourcersType = AllResourcersType + array{HardBread_Type}
                set AllResourcersAmmount = AllResourcersAmmount + array{ResourcersAmmount*ResEfficiency}
            else:
                #Water
                if(RandomResource<=Water_BaseProb, ResEfficiency := AllPickaxeEfficiency[WATER]):
                    set AllResourcersType = AllResourcersType + array{WATER}
                    if(NumResource=1):
                        set AllResourcersAmmount = AllResourcersAmmount + array{ResourcersAmmount*ResEfficiency}
                    else:
                        set AllResourcersAmmount = AllResourcersAmmount + array{ResourcersAmmount*ResEfficiency*MultipleResourceGather}
                #Flour
                else if(RandomResource<=Flour_BaseProb, ResEfficiency := AllPickaxeEfficiency[FLOUR]):
                    set AllResourcersType = AllResourcersType + array{FLOUR}
                    if(NumResource=1):
                        set AllResourcersAmmount = AllResourcersAmmount + array{ResourcersAmmount*ResEfficiency}
                    else:
                        set AllResourcersAmmount = AllResourcersAmmount + array{ResourcersAmmount*ResEfficiency*MultipleResourceGather}
                #Cheese
                else if(RandomResource<=Cheese_BaseProb, ResEfficiency := AllPickaxeEfficiency[CHEESE]):
                    set AllResourcersType = AllResourcersType + array{CHEESE}
                    if(NumResource=1):
                        set AllResourcersAmmount = AllResourcersAmmount + array{ResourcersAmmount*ResEfficiency}
                    else:
                        set AllResourcersAmmount = AllResourcersAmmount + array{ResourcersAmmount*ResEfficiency*MultipleResourceGather}
                #Pimentos
                else if(RandomResource<=Pimentos_BaseProb, ResEfficiency := AllPickaxeEfficiency[PIMENTOS]):
                    set AllResourcersType = AllResourcersType + array{PIMENTOS}
                    if(NumResource=1):
                        set AllResourcersAmmount = AllResourcersAmmount + array{ResourcersAmmount*ResEfficiency}
                    else:
                        set AllResourcersAmmount = AllResourcersAmmount + array{ResourcersAmmount*ResEfficiency*MultipleResourceGather}

        #Save Resourcers
        AllLimitsReached :=SaveResourcers(Agent,AllResourcersType,AllResourcersAmmount)
        #var PlayerData : player_data_table = SaveTuple(0)

        for(Index:=0..AllResourcersType.Length-1, Type := AllResourcersType[Index], Amount := AllResourcersAmmount[Index]):
            if(Player:=player[Agent]):
                SetNewPlayerDataCache(GameManager.QuestsAndOrdersDatabase.AddAgentQuest_ProgressV2(Player,PlayerDataCache,Amount,Type)(4))
        RefreshInventoryUI(PlayerDataCache)
        if(IsOnUKTutorial?, UKTutorial := MaybeUKTutorial?):
            UKTutorial.ResourcersGatherEvent(AllResourcersType,AllResourcersAmmount)
        return (TotalDamage,AllResourcersType,AllResourcersAmmount,AllLimitsReached,IsCritical)


    AwardAFKResourcers(Agent : agent) : void =
        var AllResourcersType : []int = array{}
        var AllResourcersAmmount : []float = array{}
        if(Player := player[Agent]):
            AmountMutantMaterial:=(CheckFinalMutantMaterialAward(PlayerDataCache.LV_Area+1,MinNum_AVGHit+1.0,1.0)/MinNum_AVGHit)*0.005
            if(AmountMutantMaterial>0.0):
                set AllResourcersType = AllResourcersType + array{MUTANTMATERIAL}
                set AllResourcersAmmount = AllResourcersAmmount + array{AmountMutantMaterial}
                set AFKRegister = AFKRegister + AmountMutantMaterial
            #Save Resourcers
            SaveTuple :=SaveResourcers(Agent,AllResourcersType,AllResourcersAmmount)
            #var NewPlayerData : player_data_table = SaveTuple(0)

            DisplayDamageUI(Agent, AllResourcersAmmount,AllResourcersType, array{false},false)
            RefreshInventoryUI(PlayerDataCache)

    #Pickaxe strength
    #Total Pickaxe strength
    RefreshPickaxeGatherUI() : void =
        if(Agent := MaybeAgentOwner?, Player := player[Agent], PickaxeUI := MaybePickaxeUpgradesUI?):
            if(not IsOnUKTutorial?):
                PickaxeUI.UpdateUI(PlayerDataCache,PickaxeTotalStrength)
            else:
                PickaxeUI.UpdateUI(player_data_table{},CalculeTotalStrength(player_data_table{}.TLV_PickaxeStrength))
        else:
            #Print("Fatal error updating PickaxeGatheUI")

    TryUpgradePickaxe_Strength(Row : int) : void =
        if:
            PickaxeUpgradesUI:=MaybePickaxeUpgradesUI?
            Agent := MaybeAgentOwner?
            PlayerOwner:=player[Agent]
            #PlayerData := PlayersDataManager.GetPlayerData[PlayerOwner]
            var Tier : int = (PickaxeUpgradesUI.CurrentPage-1)*5+Row
            var TLV_PickaxeStrength : []int = PlayerDataCache.TLV_PickaxeStrength
            Level := TLV_PickaxeStrength[Tier]
            BuyModeMultiplier := PickaxeUpgradesUI.BuyModeMultiplier
        then:
            if(not IsOnUKTutorial?):
                if:
                    Level<>-1, Level<MaxTierLevel
                then:
                    TupleProduction:=CalculePickaxeUpgradeDamage(Tier,Level,BuyModeMultiplier)
                    if:
                        NextCost:=TupleProduction(1)
                        PlayerDataCache.Inventory_Sapphires >= NextCost
                        NewTableWithoutSapphires:=PlayersDataManager.SetSapphires(PlayerDataCache,PlayerDataCache.Inventory_Sapphires-NextCost)
                        set TLV_PickaxeStrength[Tier] = Level+BuyModeMultiplier
                    then:
                        Log("Production NextCost: {NextCost} BuyModeMultiplier: {BuyModeMultiplier}",LOG_LEVEL_UIs)
                        if(TLV_PickaxeStrength[Tier+1]=-1, set TLV_PickaxeStrength[Tier+1] = 0):
                            Log("New Tier: {Tier} unlocked",LOG_LEVEL_UIs)
                        if(TLV_PickaxeStrength[Tier]>MaxTierLevel, set TLV_PickaxeStrength[Tier] = MaxTierLevel): #Check if max level was reached
                        #SAVE NEW DATA
                        NewTableCompleted:=PlayersDataManager.SetTLV_PickaxeStrength(NewTableWithoutSapphires,TLV_PickaxeStrength)
                        SetNewPlayerDataCache(NewTableCompleted)
                        #PlayersDataManager.SaveAllNewData(PlayerOwner, PlayerData, NewTableCompleted, Self)
                        PickaxeStrength := CalculeTotalStrength(TLV_PickaxeStrength)
                        set PickaxeTotalStrength = PickaxeStrength
                        PickaxeUpgradesUI.UpdateUI(NewTableCompleted,PickaxeStrength)
                        SetAllPasiveResourcers()
                        RefreshInventoryUI(NewTableCompleted)
                        GameManager.VFXManager.AwardAccolade(PICKAXE_ACCOLADE,Agent)
                        CheckQuestProgress(1.0,IMPROVEPICKAXE_TUTORIAL_TYPE)
                        spawn:
                            GameManager.VFXManager.PlaySound(SMALLOBJETIVE_SOUND,0.0,option{Agent},false)
                    else:
                        spawn:  
                            GameManager.VFXManager.PlaySound(ERROR_SOUND,0.0,option{Agent},false)
                else:
                    spawn:
                        GameManager.VFXManager.PlaySound(ERROR_SOUND,0.0,option{Agent},false)
            else:
                if(Tier=0, UKTutorial := MaybeUKTutorial?):
                    UKTutorial.PickUpgradedEvent()
                    PickaxeUpgradesUI.Remove_CanvasWithoutMessage()
                else:
                    spawn:
                        GameManager.VFXManager.PlaySound(ERROR_SOUND,0.0,option{Agent},false)

    #Efficiency
    TryUpgradePickaxe_Efficiency(Row : int) : void =
        ##Print("Trying to upgrade efficiency")
        if:
            PickaxeUpgradesUI:=MaybePickaxeUpgradesUI?
            Agent := MaybeAgentOwner?
            PlayerOwner:=player[Agent]
            #PlayerData := PlayersDataManager.GetPlayerData[PlayerOwner]
            var ArrayEfficiency_LVs : []int = PlayerDataCache.LV_PickaxeEfficiency 
            var TLV_PickaxeStrength : []int = PlayerDataCache.TLV_PickaxeStrength
            BuyModeMultiplier := PickaxeUpgradesUI.BuyModeMultiplier
        then:
            var ResEfficiencyLevel : int = 0
            if:
                AuxLevel := ArrayEfficiency_LVs[Row]
                set ResEfficiencyLevel = AuxLevel
                ResEfficiencyLevel < PickaxeMaxEfficiencyLV
            then:
                TupleEfficiency:=CalculePickaxeResourceEfficiency(ResEfficiencyLevel,Row,BuyModeMultiplier)
                var CurrentEfficiency : float = TupleEfficiency(0)
                NextCost:=TupleEfficiency(1)
                Difference:=TupleEfficiency(2)
                if:
                    PlayerDataCache.Inventory_Sapphires >= NextCost
                then:
                    SetNewPlayerDataCache(PlayersDataManager.SetSapphires(PlayerDataCache,PlayerDataCache.Inventory_Sapphires-NextCost))
                    var NewLevel : int = ResEfficiencyLevel + BuyModeMultiplier
                    if(NewLevel>PickaxeMaxEfficiencyLV):
                        set NewLevel = PickaxeMaxEfficiencyLV
                    if(set ArrayEfficiency_LVs[Row] = NewLevel):
                        #SAVE NEW DATA
                        SetNewPlayerDataCache(PlayersDataManager.SetTLV_PickaxeEfficiency(PlayerDataCache,ArrayEfficiency_LVs))
                        #PlayersDataManager.SaveAllNewData(PlayerOwner, PlayerData, NewTableCompleted, Self)
                        PickaxeStrength := CalculeTotalStrength(TLV_PickaxeStrength)
                        SetPickaxeEfficiency(Agent,PlayerDataCache,Row)
                        PickaxeUpgradesUI.UpdateUI(PlayerDataCache,PickaxeStrength)
                        RefreshInventoryUI(PlayerDataCache)
                        GameManager.VFXManager.AwardAccolade(PICKAXE_ACCOLADE,Agent)
                        spawn:
                            GameManager.VFXManager.PlaySound(SMALLOBJETIVE_SOUND,0.0,option{Agent},false)
                    SetAllPasiveResourcers()
                else:
                    spawn:
                        GameManager.VFXManager.PlaySound(ERROR_SOUND,0.0,option{Agent},false)
            else:
                spawn:
                    GameManager.VFXManager.PlaySound(ERROR_SOUND,0.0,option{Agent},false)

    SetScrollPickaxeEfficiency(PlayerData : player_data_table) : void =
        set Secret_PickaxeEfficiency = 1.0
        AllSecrets_Granted:=PlayerData.AllSecrets_Granted
        for(IsSecret_Granted : AllSecrets_Granted, IsSecret_Granted?):
            set Secret_PickaxeEfficiency = Secret_PickaxeEfficiency + PickaxeBaseScrollEfficiency
            

    SetPickaxeEfficiency(Agent : agent, PlayerData : player_data_table, Resource : int) : void =
        if:
            ResourceEfficiencyLV := PlayerData.LV_PickaxeEfficiency[Resource]
        then:
            ResourceEfficiency := 1.0+(CalculePickaxeResourceEfficiency(ResourceEfficiencyLV, Resource, 1)(0)/100.0)
            if(set AllPickaxeEfficiency[Resource]=ResourceEfficiency): 
            

    GetPackPackLimit(Resource : int) : float =
        var Limit : float = 0.0
        if(Value:=AllBackpackLimits[Resource]):
            set Limit = Value
        return Limit

    GenerateAllBackPackLimits(PlayerData : player_data_table) : void =
        for(Resource:=0..PlayerData.AllLV_BackpackLimits.Length-1): 
            if(Level := PlayerData.AllLV_BackpackLimits[Resource]):
                var Limit : float = 0.0
                TupleValue := CalculeBackpackUpgrade(Level,Resource)
                set Limit = TupleValue(0)
                ##Print("Limit of resource {Resource} is {Limit}")
                if(set AllBackpackLimits[Resource] = Limit):
            else:
                PrintFatalError("Backpack limit {Resource} can not was generated")

    SaveResourcers(Agent : agent, AllResourcersType : []int, AllResourcersAmmounts : []float, ?MustUpdateUI : logic = true, ?AutoBackpack : logic = false) : <#tuple(player_data_table,#>[]logic =
        #var PlayerData : player_data_table = player_data_table{}
        var AllLimitReached : []logic = array{}
        var AllAmountsTryingToAdd : []float = AllResourcersAmmounts
        if(Player := player[Agent]):
            #Check for autobackpack
            if(AutoBackpack?):
                Log("AutoBackpack enabled!",LOG_LEVEL_DEBUG7)
                for(Index:=0..AllResourcersType.Length-1, ResType := AllResourcersType[Index], Amount := AllResourcersAmmounts[Index]):
                    Log("AutoBackpackTrying to add ResType: {ResType}, Amount: {Amount}",LOG_LEVEL_DEBUG7)
                AllResourceLimits:=FindBackpackLimits(AllAmountsTryingToAdd,AllResourcersType)


                for(ResOverLimit : AllResourceLimits, ResOverLimit=WATER or ResOverLimit=FLOUR or ResOverLimit=CHEESE or ResOverLimit=PIMENTOS, AmountTryingToAdd := AllAmountsTryingToAdd[ResOverLimit]):
                    Log("ResOverLimit: {ResOverLimit} over limit! Upgrading Backpack",LOG_LEVEL_DEBUG7)
                    AutoBackpackAmount_Remaining:=TryBackpackUpgrade(ResOverLimit, ?AutoBackpackAmount:=AmountTryingToAdd)
                    if(set AllAmountsTryingToAdd[ResOverLimit] = AutoBackpackAmount_Remaining):
                        Log("Resource {ResOverLimit} Remaining: {AutoBackpackAmount_Remaining}!",LOG_LEVEL_DEBUG7)
            #Add resourcers
            for(Index:=0..AllResourcersType.Length-1, Type:=AllResourcersType[Index]):
                if(var Ammount : float = AllAmountsTryingToAdd[Index]):
                    var ResourceLimit : float = 0.0
                    var WasOverLimit : logic = false
                    if(ResLimit := AllBackpackLimits[Type]):
                        set ResourceLimit = ResLimit
                        ##Print("ResLimit: {ResLimit}")
                    #else:
                    #    if(Type<>MUTANTMATERIAL,Type<>):
                    #    PrintWarning("AllBackpackLimits limit : {Type} not found!")
                    #Add info limit to array  
                    AddResourcersTuple:=PlayersDataManager.AddResourcers(PlayerDataCache,Ammount,Type,ResourceLimit)
                    SetNewPlayerDataCache(AddResourcersTuple(0))
                    set AllLimitReached = AllLimitReached + array{AddResourcersTuple(1)}
                #Log("DataToUpdate.Inventory_Water is {DataToUpdate.Inventory_Water}, DataToUpdate.Inventory_Flour: {DataToUpdate.Inventory_Flour}", LOG_LEVEL_UIs)
            SetNewPlayerDataCache(PlayerDataCache)
            #PlayersDataManager.SaveAllNewData(Player, OldData, DataToUpdate, Self)
            #set PlayerData = DataToUpdate 
        return <#(PlayerData,#>AllLimitReached#)

    #Find Limits for autobackpack
    FindBackpackLimits(AllAmmounts : []float, AllTypes : []int):[]int = 
        var AllOverLimit : []int = array{}
        for(Index := 0..AllTypes.Length-1, Resource := AllTypes[Index]):
            var CurrentAmount  : float = 0.0
            if(Resource=WATER):
                set CurrentAmount = PlayerDataCache.Inventory_Water
            else if(Resource=FLOUR):
                set CurrentAmount = PlayerDataCache.Inventory_Flour
            else if(Resource=CHEESE):
                set CurrentAmount = PlayerDataCache.Inventory_Cheese
            else if(Resource=PIMENTOS):
                set CurrentAmount = PlayerDataCache.Inventory_Pimentos

            Log("Checking limit of {Resource}, CurrentAmount: {GetOptimizedMagnitudeLetter(CurrentAmount)}", LOG_LEVEL_DEBUG7)

            if(Limit := AllBackpackLimits[Resource], AmountToAdd := AllAmmounts[Resource]):
                TotalToSet := CurrentAmount + AmountToAdd
                if(Limit < TotalToSet):
                    set AllOverLimit = AllOverLimit + array{Resource}
                    Log("Resource: {Resource} over limit!",LOG_LEVEL_DEBUG7)
                else:
                    Log("Was not over limit Resource: {Resource} Because; (AmountToAdd: {GetOptimizedMagnitudeLetter(AmountToAdd)} + CurrentAmount: {GetOptimizedMagnitudeLetter(CurrentAmount)}) < Limit: {GetOptimizedMagnitudeLetter(Limit)}",LOG_LEVEL_DEBUG7)
        return AllOverLimit

    DisplayDamageUI(Agent : agent, Ammounts : []float, Types : []int, LimitsReached : []logic,CriticalStrike : logic) : void =
        if(ActionInfoUI := MaybeActionInfoUI?):
            ActionInfoUI.DisplayGather(Ammounts, Types, LimitsReached,CriticalStrike)

                <#
    SetPickaxeInfo(Agent : agent) : void =
        Log("UpdatePickaxeInfo updated", LOG_LEVEL_UIs)
        if(Player := player[Agent],PlayerData:=PlayersDataManager.GetPlayerData[Player]):
            TotalStrength:=MainFormula_TotalPickaxeStrength(PlayerData.TLV_PickaxeStrength)
            if:
                WaterEfficiencyLevel := PlayerData.LV_PickaxeEfficiency[0]
                FlourEfficiencyLevel := PlayerData.LV_PickaxeEfficiency[1]
                CheeseEfficiencyLevel := PlayerData.LV_PickaxeEfficiency[2]
                PimentosEfficiencyLevel := PlayerData.LV_PickaxeEfficiency[3]
                MutantEfficiencyLevel := PlayerData.LV_PickaxeEfficiency[4]
            then:

                Log("TotalStrength loaded: {TotalStrength}", LOG_LEVEL_UIs)

                WaterEfficiency:=MainFormula_PickaxeEfficiency(WaterEfficiencyLevel*1.0)
                FlourEfficiency:=MainFormula_PickaxeEfficiency(FlourEfficiencyLevel*1.0)
                CheeseEfficiency:=MainFormula_PickaxeEfficiency(CheeseEfficiencyLevel*1.0)
                PimentosEfficiency:=MainFormula_PickaxeEfficiency(PimentosEfficiencyLevel*1.0)
                MutantEfficiency:=MainFormula_PickaxeEfficiency(MutantEfficiencyLevel*1.0)

                  
                WeakPointMult:=MainFormula_WeakPointMult(PlayerData.LV_PickaxeWeakPointMult)
                CoalLuck:=MainFormula_CoalProb(PlayerData.LV_PickaxeLuck)
                CriticalDamage:=MainFormula_PickaxeCriticalDamage(PlayerData.LV_PickaxeCriticalDamage)
                CriticalProb:=MainFormula_PickaxeCriticalProb(PlayerData.LV_PickaxeCriticalProb)

                set PickaxeMultiResource=PlayerData.LV_PickaxeMultiResource
                set PickaxeTotalStrength=TotalStrength
                set PickaxeCoalLuck=CoalLuck
           
                set PickaxeCriticalDamage = CriticalDamage
                set PickaxeCriticalProb = CriticalProb
                set PickaxeWeakPointMult =WeakPointMult
                set PickaxeMultiResource = PickaxeMultiResource

                WaterEfficiency:=MainFormula_PickaxeEfficiency(WaterEfficiencyLevel*1.0)
                FlourEfficiency:=MainFormula_PickaxeEfficiency(FlourEfficiencyLevel*1.0)
                CheeseEfficiency:=MainFormula_PickaxeEfficiency(CheeseEfficiencyLevel*1.0)
                PimentosEfficiency:=MainFormula_PickaxeEfficiency(PimentosEfficiencyLevel*1.0)
                MutantEfficiency:=MainFormula_PickaxeEfficiency(MutantEfficiencyLevel*1.0)
                
                if:
                    set AllPickaxeEfficiency[WATER]=WaterEfficiency
                    set AllPickaxeEfficiency[FLOUR]=FlourEfficiency
                    set AllPickaxeEfficiency[CHEESE]=CheeseEfficiency
                    set AllPickaxeEfficiency[PIMENTOS]=PimentosEfficiency
                    set AllPickaxeEfficiency[MUTANTMATERIAL]=MutantEfficiency#>

    RefreshInventoryUI(PlayerData : player_data_table) : void =
        if:
            MainInventoryUI := MaybeMainInventoryUI?
        then:
            WaterLimit := GetPackPackLimit(WATER)
            FlourLimit := GetPackPackLimit(FLOUR)
            CheeseLimit := GetPackPackLimit(CHEESE)
            PimentosLimit := GetPackPackLimit(PIMENTOS)
            #MutantMaterialLimit := GetPackPackLimit(MUTANTMATERIAL)
            #CoalLimit := GetPackPackLimit(COAL)
            LimitArray:=array{false,false,false,false,false,false,false}
            MainInventoryUI.RefreshInventory(PlayerData,array{WaterLimit,FlourLimit,CheeseLimit,PimentosLimit},LimitArray)
            #MainInventoryUI.RefreshAll(PlayerData,array{WaterLimit,FlourLimit,CheeseLimit,PimentosLimit},LimitArray)

    #---------------------

    #PLAYER REFERENCE DEVICE
    SetPlayerReference(InPlayerReferenceDevice : player_reference_device) : void =
        set OwnerReferenceDevice = InPlayerReferenceDevice
    
    CloseAllTutorialAffectedUI() : void =
        if:          
            PickaxeUpgradesUI:=MaybePickaxeUpgradesUI?
            BBakersUpgradesU:=MaybeBBakersUpgradesUI?
            #BBInfoUI:=MaybeBBInfoUI?
            BackpackUpgradesUI:=MaybeBackpackUpgradesUI?
            PickaxePerksUI:=MaybePickaxePerksUI?
            DepositsUpgradeUI:=MaybeDepositsUpgradeUI?
            WeaponsUpgradesUI:=MaybeWeaponsUpgradesUI?
            DropCoalUI:=MaybeDropCoalUI?
            TokensUI:=MaybeTokensUI?
        then:
            PetsManager.CloseAllUI()
            PickaxeUpgradesUI.Remove_CanvasWithoutMessage()
            BBakersUpgradesU.Remove_CanvasWithoutMessage()
            #BBInfoUI.Remove_CanvasWithoutMessage()
            BackpackUpgradesUI.Remove_CanvasWithoutMessage()
            PickaxePerksUI.Remove_CanvasWithoutMessage()
            DepositsUpgradeUI.Remove_CanvasWithoutMessage()
            WeaponsUpgradesUI.Remove_CanvasWithoutMessage()
            DropCoalUI.Remove_CanvasWithoutMessage()
            TokensUI.DT_Remove_CanvasWithoutMessage()
            TokensUI.HT_Remove_CanvasWithoutMessage()
            #HexaGameManager.CloseAllUI()
    #SETTERS
    CloseAllUI(?IsComplete : logic = true) : void =
        if:           
            MasteryUI:=MaybeMasteryUI?
            ActionInfoUI:=MaybeActionInfoUI?
            DungeonUI:=MaybeDungeonUI?
            MainInventoryUI:=MaybeMainInventoryUI?
            EntityHealthUI:=MaybeEntityHealthUI?
            PickaxeUpgradesUI:=MaybePickaxeUpgradesUI?
            #ResourcersUI:=MaybeResourcersUI?
            MessagesUI:=MaybeMessagesUI?
            PCSBakersUpgradesUI:=MaybePCSBakersUpgradesUI?
            BBakersUpgradesU:=MaybeBBakersUpgradesUI?
            CBBakersUpgradesUI:=MaybeCBBakersUpgradesUI?
            PCSBInfoUI:=MaybePCSBInfoUI?
            CBBInfoUI:=MaybeCBBInfoUI?
            BBInfoUI:=MaybeBBInfoUI?
            DroneUpgrades:=MaybeDroneUpgrades?
            BackpackUpgradesUI:=MaybeBackpackUpgradesUI?
            PickaxePerksUI:=MaybePickaxePerksUI?
            DepositsUpgradeUI:=MaybeDepositsUpgradeUI?
            FastTravelUI:=MaybeFastTravelUI?
            WeaponsUpgradesUI:=MaybeWeaponsUpgradesUI?
            DropCoalUI:=MaybeDropCoalUI?
            OptionsMenuUI:=MaybeOptionsMenuUI?
            LanguageUI:=MaybeLanguageUI?
            TokensUI:=MaybeTokensUI?
        then:
            if(IsComplete?):
                PetsManager.CloseAllUI()
                MainInventoryUI.Remove_CanvasWithoutMessage()
                ActionInfoUI.Remove_CanvasWithoutMessage()
                EntityHealthUI.Remove_CanvasWithoutMessage()

            MasteryUI.Remove_CanvasWithoutMessage()
            DungeonUI.Remove_CanvasWithoutMessage()
            PickaxeUpgradesUI.Remove_CanvasWithoutMessage()
            #ResourcersUI.Remove_CanvasWithoutMessage()
            MessagesUI.Remove_CanvasWithoutMessage()
            PCSBakersUpgradesUI.Remove_CanvasWithoutMessage()
            BBakersUpgradesU.Remove_CanvasWithoutMessage()
            CBBakersUpgradesUI.Remove_CanvasWithoutMessage()
            PCSBInfoUI.Remove_CanvasWithoutMessage()
            CBBInfoUI.Remove_CanvasWithoutMessage()
            BBInfoUI.Remove_CanvasWithoutMessage()
            DroneUpgrades.Remove_CanvasWithoutMessage()
            BackpackUpgradesUI.Remove_CanvasWithoutMessage()
            PickaxePerksUI.Remove_CanvasWithoutMessage()
            DepositsUpgradeUI.Remove_CanvasWithoutMessage()
            FastTravelUI.Remove_CanvasWithoutMessage()
            WeaponsUpgradesUI.Remove_CanvasWithoutMessage()
            DropCoalUI.Remove_CanvasWithoutMessage()
            OptionsMenuUI.Remove_CanvasWithoutMessage()
            LanguageUI.Remove_CanvasWithoutMessage()
            TokensUI.DT_Remove_CanvasWithoutMessage()
            TokensUI.HT_Remove_CanvasWithoutMessage()

            HexaGameManager.CloseAllUI()
        else:
            PrintFatalError("On DebugAllui")

    ForceLanguageSelection() : void =
        if:           
            MasteryUI:=MaybeMasteryUI?
            ActionInfoUI:=MaybeActionInfoUI?
            DungeonUI:=MaybeDungeonUI?
            MainInventoryUI:=MaybeMainInventoryUI?
            EntityHealthUI:=MaybeEntityHealthUI?
            PickaxeUpgradesUI:=MaybePickaxeUpgradesUI?
            #ResourcersUI:=MaybeResourcersUI?
            MessagesUI:=MaybeMessagesUI?
            PCSBakersUpgradesUI:=MaybePCSBakersUpgradesUI?
            BBakersUpgradesU:=MaybeBBakersUpgradesUI?
            CBBakersUpgradesUI:=MaybeCBBakersUpgradesUI?
            PCSBInfoUI:=MaybePCSBInfoUI?
            CBBInfoUI:=MaybeCBBInfoUI?
            BBInfoUI:=MaybeBBInfoUI?
            DroneUpgrades:=MaybeDroneUpgrades?
            BackpackUpgradesUI:=MaybeBackpackUpgradesUI?
            PickaxePerksUI:=MaybePickaxePerksUI?
            DepositsUpgradeUI:=MaybeDepositsUpgradeUI?
            FastTravelUI:=MaybeFastTravelUI?
            WeaponsUpgradesUI:=MaybeWeaponsUpgradesUI?
            DropCoalUI:=MaybeDropCoalUI?
            OptionsMenuUI:=MaybeOptionsMenuUI?
            LanguageUI:=MaybeLanguageUI?
            TokensUI:=MaybeTokensUI?
        then:
            MasteryUI.Remove_CanvasWithoutMessage()
            DungeonUI.Remove_CanvasWithoutMessage()
            PickaxeUpgradesUI.Remove_CanvasWithoutMessage()
            #ResourcersUI.Remove_CanvasWithoutMessage()
            #MessagesUI.Remove_CanvasWithoutMessage()
            PCSBakersUpgradesUI.Remove_CanvasWithoutMessage()
            BBakersUpgradesU.Remove_CanvasWithoutMessage()
            CBBakersUpgradesUI.Remove_CanvasWithoutMessage()
            PCSBInfoUI.Remove_CanvasWithoutMessage()
            CBBInfoUI.Remove_CanvasWithoutMessage()
            BBInfoUI.Remove_CanvasWithoutMessage()
            DroneUpgrades.Remove_CanvasWithoutMessage()
            BackpackUpgradesUI.Remove_CanvasWithoutMessage()
            PickaxePerksUI.Remove_CanvasWithoutMessage()
            DepositsUpgradeUI.Remove_CanvasWithoutMessage()
            FastTravelUI.Remove_CanvasWithoutMessage()
            WeaponsUpgradesUI.Remove_CanvasWithoutMessage()
            DropCoalUI.Remove_CanvasWithoutMessage()
            OptionsMenuUI.Remove_CanvasWithoutMessage()
            #LanguageUI.Remove_CanvasWithoutMessage()
            TokensUI.DT_Remove_CanvasWithoutMessage()
            TokensUI.HT_Remove_CanvasWithoutMessage()
            LanguageUI.Display_Canvas()
        else:
            PrintFatalError("On DebugAllui")


    SetNewOwner(NewAgentOwner : agent, NewPlayerData : player_data_table) : void =
        set AsyncFame = -1.0
        set AsyncFameDifference = -1.0
        set MaybeAgentOwner  = option{NewAgentOwner}
        set PlayerDataCache = NewPlayerData 
        set IsOnUKTutorial = false
        #UIS
        if:           
            BBakersUpgradesUI:=MaybeBBakersUpgradesUI?
            CBBakersUpgradesUI:=MaybeCBBakersUpgradesUI?
            PCSBakersUpgradesUI:=MaybePCSBakersUpgradesUI?

            PCSBInfoUI:=MaybePCSBInfoUI?
            CBBInfoUI:=MaybeCBBInfoUI?
            BBInfoUI:=MaybeBBInfoUI?

            DepositsUpgradeUI:=MaybeDepositsUpgradeUI?
            BackpackUpgradesUI:=MaybeBackpackUpgradesUI?

            ActionInfoUI:=MaybeActionInfoUI?
            EntityHealthUI:= MaybeEntityHealthUI?
            MainInventoryUI := MaybeMainInventoryUI?
            PickaxeUpgradesUI := MaybePickaxeUpgradesUI? 
            #ResourcersUI := MaybeResourcersUI?
            MessagesUI := MaybeMessagesUI? 
            PickaxePerksUI := MaybePickaxePerksUI?
            #Weapon Upgrades
            WeaponsUpgradesUI:=MaybeWeaponsUpgradesUI?
            DungeonUI:=MaybeDungeonUI?
            DropCoalUI:=MaybeDropCoalUI?
            TokensUI:=MaybeTokensUI?
            OptionsMenuUI:=MaybeOptionsMenuUI?

            FastTravelUI:=MaybeFastTravelUI?
        then:
            BBakersUpgradesUI.SetAgent(NewAgentOwner)
            CBBakersUpgradesUI.SetAgent(NewAgentOwner)
            PCSBakersUpgradesUI.SetAgent(NewAgentOwner)
            DepositsUpgradeUI.SetAgent(NewAgentOwner)
            BackpackUpgradesUI.SetAgent(NewAgentOwner)
            ActionInfoUI.SetAgent(NewAgentOwner)
            EntityHealthUI.SetAgent(NewAgentOwner)
            #EntityHealthUI.Display_Canvas()
            MainInventoryUI.SetAgent(NewAgentOwner)
            PickaxeUpgradesUI.SetAgent(NewAgentOwner)
            #ResourcersUI.SetAgent(NewAgentOwner)
            MessagesUI.SetAgent(NewAgentOwner)
            PickaxePerksUI.SetAgent(NewAgentOwner)
            WeaponsUpgradesUI.SetAgent(NewAgentOwner)
            DungeonUI.SetAgent(NewAgentOwner)
            DropCoalUI.SetAgent(NewAgentOwner)
            TokensUI.SetAgent(NewAgentOwner)
            FastTravelUI.SetAgent(NewAgentOwner)
            OwnerReferenceDevice.Register(NewAgentOwner)
            HexaGameManager.SetAgent(NewAgentOwner)
            OptionsMenuUI.SetAgent(NewAgentOwner)
            DisplayBasicUIs()
        else:
            PrintFatalError("Error on SetNewOwner")
        spawn:
            LoadPlayerStats()
        ResetTemporalStats()
        #GameManager.GeneralMobsManager.InitPlayer(NewAgentOwner)
        CheckMusicPlay(BAKERSMUSIC_SOUND)

    ResetTemporalStats() : void = 
        set AFKRegister = 0.0

    ClearAFKRegister() : void =
        set AFKRegister = 0.0

    SetMainDepositsProps(AllMainDeposits : []creative_prop) : void =
        if:
            Prop1:=AllMainDeposits[0],
            set MainWaterDeposit_AmmountProp = Prop1,
            Prop2:=AllMainDeposits[1],
            set MainFlourDeposit_AmmountProp = Prop2,
            Prop3:=AllMainDeposits[2],
            set MainCheeseDeposit_AmmountProp = Prop3,
            Prop4:=AllMainDeposits[3],
            set MainPimentosDeposit_AmmountProp = Prop4
        else:
            #Print("Fatal error on SetMainDepositsProps")

    SetSuperEnergyDevices(Timer : timer_device, InOverPoweredSign : creative_prop) : void =
        set OverPoweredSign = InOverPoweredSign
        set SuperEnergyRemaining_Timer = Timer

    SetBillBoards(AllBillboards : []billboard_device) : void =
        if:
            WaterDeposit2 := AllBillboards[0]
            set WaterDeposit_Billbord = WaterDeposit2
            FlourDeposit := AllBillboards[1]
            set FlourDeposit_Billbord = FlourDeposit
            CheeseDeposit := AllBillboards[2]
            set CheeseDeposit_Billbord = CheeseDeposit
            PimentosDeposit := AllBillboards[3]
            set PimentosDeposit_Billbord = PimentosDeposit
            DroneDeposit := AllBillboards[4]
            set DroneDeposit_Billbord = DroneDeposit
            DroneShip := AllBillboards[5]
            set DroneShip_Billbord = DroneShip
        else:
            #Print("FATAL ERROR setting billboards")

    SetValves(Tuple : tuple([]creative_prop,[]button_device,[]button_device)) : void = 
        set AllProps_Valves = Tuple(0)
        set AllButtons_Valves = Tuple(1)
        AllSimplifiedButtons := Tuple(2)

        if:
            ValveButton_B := AllSimplifiedButtons[0]
            ValveButton_CB := AllSimplifiedButtons[1]
            ValveButton_PCS := AllSimplifiedButtons[2]
        then:
            set Simplified_ValveButtons_B = ValveButton_B
            set Simplified_ValveButtons_CB = ValveButton_CB
            set Simplified_ValveButtons_PCS = ValveButton_PCS
        #Print("AllButtons_Valves: {AllButtons_Valves.Length}")
        #Print("AllProps_Valves: {AllProps_Valves.Length}")
        InitValves()
    SetTeleportationDevices(Button : button_device, Prop : creative_prop) : void =
        set Prop_PersonalTeleporter = Prop
        set Button_PersonalTeleporter = Button

    SetBakersProps(AllBakersProps : []creative_prop) : void =
        if:
            Prop1:=AllBakersProps[0]
            set BB_Prop = Prop1
            Prop2:=AllBakersProps[1]
            set CBB_Prop = Prop2
            Prop3:=AllBakersProps[2]
            set PCSB_Prop = Prop3
            Prop4:=AllBakersProps[3]
            set DroneProp = Prop4
        else:
            PrintFatalError("Fatal error on SetAllBakersProps")

    SetFunnelButton(FunnelButton : button_device) : void =
        set Button_ResFunnel = FunnelButton 

    SetUpgradeButtons(AllButton_ComputerButtons : []button_device) : void =
        if:
            #Baker upgrades
            Button0 := AllButton_ComputerButtons[0]
            set Button_BBUpgrades = Button0
            Button1 := AllButton_ComputerButtons[1]
            set Button_CBBUpgrades = Button1
            Button2 := AllButton_ComputerButtons[2]
            set Button_PCSBUpgrades = Button2
            #Info computer Buttons
            Button3 := AllButton_ComputerButtons[3]
            set Button_InfoBB = Button3
            Button4 := AllButton_ComputerButtons[4]
            set Button_InfoCBB = Button4
            Button5 := AllButton_ComputerButtons[5]
            set Button_InfoPCSB = Button5 
            #Deposits
            Button6 := AllButton_ComputerButtons[6]
            set Button_DepositsUpgrades = Button6
            #Drone upgrades
            Button7 := AllButton_ComputerButtons[7]
            set Button_DroneUpgrades = Button7
            #Drone ship
            Button8 := AllButton_ComputerButtons[8]
            set Button_DroneShip = Button8
        else:
            PrintFatalError("on SetUpgradeButtons")
            #Backpack upgrades
            
    SetAllMainDepositLimits(Agent : agent, AllLimits : []int) : void =
        if:
            WaterDepositLimitLV := AllLimits[WATER]
            FlourDepositLimitLV := AllLimits[FLOUR]
            CheeseDepositLimitLV := AllLimits[CHEESE]
            PimentosDepositLimitLV := AllLimits[PIMENTOS]
        then:
            set AllLV_MainDeposits = AllLimits

            WaterDepositLimit:=CalculeDepositUpgradeLimit(WaterDepositLimitLV,WATER)(0)
            FlourDepositLimit:=CalculeDepositUpgradeLimit(FlourDepositLimitLV,FLOUR)(0)
            CheeseDepositLimit:=CalculeDepositUpgradeLimit(CheeseDepositLimitLV,CHEESE)(0)
            PimentosDepositLimit:=CalculeDepositUpgradeLimit(PimentosDepositLimitLV,PIMENTOS)(0)
            if:
                set AllMainDeposit_Limits[WATER]  = WaterDepositLimit
                set AllMainDeposit_Limits[FLOUR]  = FlourDepositLimit
                set AllMainDeposit_Limits[CHEESE]  = CheeseDepositLimit
                set AllMainDeposit_Limits[PIMENTOS]  = PimentosDepositLimit
            UpdateAllMainDeposits()
        else:
            PrintFatalError("on SetUpgradeButtons")

    SetAllPasiveResourcers() : void =
        if:
            Agent := MaybeAgentOwner?
            WaterResEfficiency := AllPickaxeEfficiency[WATER]
            FlourResEfficiency := AllPickaxeEfficiency[FLOUR]
            CheeseResEfficiency := AllPickaxeEfficiency[CHEESE]
            PimentosResEfficiency := AllPickaxeEfficiency[PIMENTOS]

            WaterPasiveLV := PlayerDataCache.AllLV_PasiveResourcers[WATER]
            FlourPasiveLV := PlayerDataCache.AllLV_PasiveResourcers[FLOUR]
            CheesePasiveLV := PlayerDataCache.AllLV_PasiveResourcers[CHEESE]
            PimentosPasiveLV := PlayerDataCache.AllLV_PasiveResourcers[PIMENTOS]
        then:
            WaterPasive:=CalculeDepositPasiveIncome(WaterPasiveLV)(0)
            FlourPasive:=CalculeDepositPasiveIncome(FlourPasiveLV)(0)
            CheesePasive:=CalculeDepositPasiveIncome(CheesePasiveLV)(0)
            PimentosPasive:=CalculeDepositPasiveIncome(PimentosPasiveLV)(0)
            
            var HBLevel : int =  PlayerDataCache.LV_Area+1
            if(HBLevel=0):
                set HBLevel = 1

            TuplePickDamage := GetBakeryPickaxeDamage(HBLevel,false,true)
            var TotalDamage : float = TuplePickDamage(0)
            var ResourcersAmmount : float = ApplyEfficiencyToDamage(TotalDamage)

            Log("WaterPasive: {WaterPasive}, TotalResourcers; {ResourcersAmmount} Total: {WaterPasive * ResourcersAmmount}",LOG_LEVEL_DEBUG5)

            if:
                set AllPasiveResourcers[WATER]  = WaterPasive * ResourcersAmmount
                set AllPasiveResourcers[FLOUR]  = FlourPasive * ResourcersAmmount
                set AllPasiveResourcers[CHEESE]  = CheesePasive * ResourcersAmmount
                set AllPasiveResourcers[PIMENTOS]  = PimentosPasive * ResourcersAmmount
            else:
                PrintFatalError("on SetUpgradeButtons")


    InitSubscribes(): void =
        #Teleportation
        Button_PersonalTeleporter.InteractedWithEvent.Subscribe(handler_random{Index:=TELEPORTER_FROM_BAKERY_AREA, MaybeGameManager := option{GameManager}}.HandlerOpenTeleporterMenu)   
        #Main Deposits
        Button_DepositsUpgrades.InteractedWithEvent.Subscribe(Display_DepositsUpgrades)
        #Main Funnel
        Button_ResFunnel.InteractedWithEvent.Subscribe(InsertResourcers)
        #Computers
        Button_BBUpgrades.InteractedWithEvent.Subscribe(Display_ComputerBBInfo)
        Button_CBBUpgrades.InteractedWithEvent.Subscribe(Display_ComputerCBBInfo)
        Button_PCSBUpgrades.InteractedWithEvent.Subscribe(Display_ComputerPCSBInfo)
        #Drone
        Button_DroneUpgrades.InteractedWithEvent.Subscribe(Display_DroneUpgradesInfo)
        #Drone ship
        Button_DroneShip.InteractedWithEvent.Subscribe(TrySendDrone)
        #Info
        Button_InfoPCSB.InteractedWithEvent.Subscribe(Display_InfoPCSB)
        Button_InfoCBB.InteractedWithEvent.Subscribe(Display_InfoCBB)
        Button_InfoBB.InteractedWithEvent.Subscribe(Display_InfoBB)

        #Simplified Valves
        if(GameManager.IsValveSimplifiedVersion?):
            Simplified_ValveButtons_B.InteractedWithEvent.Subscribe(handler_upgrades{BakerType:=BB_TYPE, MaybeBakeryMachinesManager := option{Self}}.HandlerTryChangeSimplifiedValveState) 
            Simplified_ValveButtons_CB.InteractedWithEvent.Subscribe(handler_upgrades{BakerType:=CBB_TYPE, MaybeBakeryMachinesManager := option{Self}}.HandlerTryChangeSimplifiedValveState) 
            Simplified_ValveButtons_PCS.InteractedWithEvent.Subscribe(handler_upgrades{BakerType:=PCSB_TYPE, MaybeBakeryMachinesManager := option{Self}}.HandlerTryChangeSimplifiedValveState) 
            for(Button : AllButtons_Valves):
                Button.Disable()
        else:
            for(Index:=0..AllButtons_Valves.Length-1, Button := AllButtons_Valves[Index]):
                Button.InteractedWithEvent.Subscribe(handler_upgrades{Index:=Index, MaybeBakeryMachinesManager := option{Self}, CraftMode := false}.HandlerTryChangeValveState) 
            Simplified_ValveButtons_B.Disable() 
            Simplified_ValveButtons_CB.Disable() 
            Simplified_ValveButtons_PCS.Disable() 

MakeBakeryMachinesManager<constructor>(
    InAgentAndBaker_Index : int,
    InAgentOwner:agent,
    InHexaGameManager : hexa_minigame_manager,
    InGameManager: game_manager)<transacts> := bakery_machines_manager :
        AgentAndBaker_Index := InAgentAndBaker_Index
        MaybeAgentOwner := option{InAgentOwner}
        GameManager:= InGameManager
        PetsManager := pets_manager{GameManager:=InGameManager}
        HexaGameManager := InHexaGameManager