using { /UnrealEngine.com/Temporary/UI }
using { /UnrealEngine.com/Temporary/SpatialMath }
using { /Verse.org/Simulation }
using { /Fortnite.com/Devices }
using { /Fortnite.com/UI }
using { /Verse.org/Colors/NamedColors }
using { /Verse.org/Colors }

# See https://dev.epicgames.com/documentation/en-us/uefn/create-your-own-device-in-verse for how to create a verse device.

# A Verse-authored creative device that can be placed in a level
Bbaker_upgrades_ui := class():

    #Warning Message
    WarningTitle_TextBlock : text_block = text_block{DefaultTextColor:=Red, DefaultShadowOffset:=ShadowOffset, DefaultShadowOpacity := DefaultTextShadowOpacity}
    Warning_TextBlock : text_block = text_block{DefaultTextColor:=Black, DefaultShadowOffset:=ShadowOffset, DefaultShadowOpacity := DefaultTextShadowOpacity}

    #Class settings
    var MaybeAgent : ?agent = false
    
    BackgroundImageSizeX : float = 1920.0
    BackgroundImageSizeY : float = 1080.0

    TopImageSizeX : float = 175.0
    TopImageSizeY : float = 80.0

    #LineSizeX : float = 1000.0
    #LineSizeY : float = 80.0 #Size of button regular
    
    #Resourcers
    EmptyTextureBlock : texture_block = texture_block{DefaultDesiredSize:=vector2{X:=0.0,Y:=0.0},DefaultImage:=Images.EmptyTexture}
    VerticalLineTextureBlock : texture_block = texture_block{DefaultDesiredSize:=vector2{X:=0.0,Y:=0.0},DefaultImage:=Images.Img_VerticalLine}

#Settings Upgrade production

    BakeryMachinesManager : bakery_machines_manager

    UPColumn1SizeX : float = 80.0
    UPColumn1SizeY : float = 80.0

    UPColumn2SizeX : float = 75.0
    UPColumn2SizeY : float = 80.0

    UPColumn3SizeX : float = 150.0
    UPColumn3SizeY : float = 80.0

    UPColumn4SizeX : float = 200.0
    UPColumn4SizeY : float = 80.0

    UPColumn5SizeX : float = 200.0
    UPColumn5SizeY : float = 80.0

    UPColumn6SizeX : float = 225.0
    UPColumn6SizeY : float = 80.0

        #Data

    UP_Row1Button : button_quiet = button_quiet{}
    UP_Row2Button : button_quiet = button_quiet{}
    UP_Row3Button : button_quiet = button_quiet{}
    UP_Row4Button : button_quiet = button_quiet{}
    UP_Row5Button : button_quiet = button_quiet{}

    UP_Row1_SapphiresTexture : texture_block = texture_block{DefaultDesiredSize:=vector2{X:=SizeSmallIconX,Y:=SizeSmallIconY}, DefaultImage:=Images.Sapphire_Icon}
    UP_Row2_SapphiresTexture : texture_block = texture_block{DefaultDesiredSize:=vector2{X:=SizeSmallIconX,Y:=SizeSmallIconY}, DefaultImage:=Images.Sapphire_Icon}
    UP_Row3_SapphiresTexture : texture_block = texture_block{DefaultDesiredSize:=vector2{X:=SizeSmallIconX,Y:=SizeSmallIconY}, DefaultImage:=Images.Sapphire_Icon}
    UP_Row4_SapphiresTexture : texture_block = texture_block{DefaultDesiredSize:=vector2{X:=SizeSmallIconX,Y:=SizeSmallIconY}, DefaultImage:=Images.Sapphire_Icon}
    UP_Row5_SapphiresTexture : texture_block = texture_block{DefaultDesiredSize:=vector2{X:=SizeSmallIconX,Y:=SizeSmallIconY}, DefaultImage:=Images.Sapphire_Icon}

    UP_PrevPageButton : button_regular = button_regular{}
    UP_NextPageButton : button_regular = button_regular{}

    CloseButton : button_loud = button_loud{}

    UPRow1TextureBlock : texture_block = texture_block{DefaultDesiredSize:=vector2{X:=SizeTextureBlockX,Y:=SizeTextureBlockY}, DefaultImage:=Images.EmptyTexture}
    UPRow2TextureBlock : texture_block = texture_block{DefaultDesiredSize:=vector2{X:=SizeTextureBlockX,Y:=SizeTextureBlockY}, DefaultImage:=Images.EmptyTexture}
    UPRow3TextureBlock : texture_block = texture_block{DefaultDesiredSize:=vector2{X:=SizeTextureBlockX,Y:=SizeTextureBlockY}, DefaultImage:=Images.EmptyTexture}
    UPRow4TextureBlock : texture_block = texture_block{DefaultDesiredSize:=vector2{X:=SizeTextureBlockX,Y:=SizeTextureBlockY}, DefaultImage:=Images.EmptyTexture}
    UPRow5TextureBlock : texture_block = texture_block{DefaultDesiredSize:=vector2{X:=SizeTextureBlockX,Y:=SizeTextureBlockY}, DefaultImage:=Images.EmptyTexture}
        
    UP_Row1_Tier_TextBlock : text_block = text_block{DefaultTextColor:=White, DefaultShadowOffset:=ShadowOffset, DefaultShadowOpacity := DefaultTextShadowOpacity}
    UP_Row2_Tier_TextBlock : text_block = text_block{DefaultTextColor:=White, DefaultShadowOffset:=ShadowOffset, DefaultShadowOpacity := DefaultTextShadowOpacity}
    UP_Row3_Tier_TextBlock : text_block = text_block{DefaultTextColor:=White, DefaultShadowOffset:=ShadowOffset, DefaultShadowOpacity := DefaultTextShadowOpacity}
    UP_Row4_Tier_TextBlock : text_block = text_block{DefaultTextColor:=White, DefaultShadowOffset:=ShadowOffset, DefaultShadowOpacity := DefaultTextShadowOpacity}
    UP_Row5_Tier_TextBlock : text_block = text_block{DefaultTextColor:=White, DefaultShadowOffset:=ShadowOffset, DefaultShadowOpacity := DefaultTextShadowOpacity}

    UP_Row1_Level_TextBlock : text_block = text_block{DefaultTextColor:=White, DefaultShadowOffset:=ShadowOffset, DefaultShadowOpacity := DefaultTextShadowOpacity}
    UP_Row2_Level_TextBlock : text_block = text_block{DefaultTextColor:=White, DefaultShadowOffset:=ShadowOffset, DefaultShadowOpacity := DefaultTextShadowOpacity}
    UP_Row3_Level_TextBlock : text_block = text_block{DefaultTextColor:=White, DefaultShadowOffset:=ShadowOffset, DefaultShadowOpacity := DefaultTextShadowOpacity}
    UP_Row4_Level_TextBlock : text_block = text_block{DefaultTextColor:=White, DefaultShadowOffset:=ShadowOffset, DefaultShadowOpacity := DefaultTextShadowOpacity}
    UP_Row5_Level_TextBlock : text_block = text_block{DefaultTextColor:=White, DefaultShadowOffset:=ShadowOffset, DefaultShadowOpacity := DefaultTextShadowOpacity}

    UP_Row1_Production_TextBlock : text_block = text_block{DefaultTextColor:=CurrentTextColor, DefaultShadowOffset:=ShadowOffset, DefaultShadowOpacity := DefaultTextShadowOpacity}
    UP_Row2_Production_TextBlock : text_block = text_block{DefaultTextColor:=CurrentTextColor, DefaultShadowOffset:=ShadowOffset, DefaultShadowOpacity := DefaultTextShadowOpacity}
    UP_Row3_Production_TextBlock : text_block = text_block{DefaultTextColor:=CurrentTextColor, DefaultShadowOffset:=ShadowOffset, DefaultShadowOpacity := DefaultTextShadowOpacity}
    UP_Row4_Production_TextBlock : text_block = text_block{DefaultTextColor:=CurrentTextColor, DefaultShadowOffset:=ShadowOffset, DefaultShadowOpacity := DefaultTextShadowOpacity}
    UP_Row5_Production_TextBlock : text_block = text_block{DefaultTextColor:=CurrentTextColor, DefaultShadowOffset:=ShadowOffset, DefaultShadowOpacity := DefaultTextShadowOpacity}

    UP_Row1_Upgrade_TextBlock : text_block = text_block{DefaultTextColor:=UpgradeTextColor, DefaultShadowOffset:=ShadowOffset, DefaultShadowOpacity := DefaultTextShadowOpacity}
    UP_Row2_Upgrade_TextBlock : text_block = text_block{DefaultTextColor:=UpgradeTextColor, DefaultShadowOffset:=ShadowOffset, DefaultShadowOpacity := DefaultTextShadowOpacity}
    UP_Row3_Upgrade_TextBlock : text_block = text_block{DefaultTextColor:=UpgradeTextColor, DefaultShadowOffset:=ShadowOffset, DefaultShadowOpacity := DefaultTextShadowOpacity}
    UP_Row4_Upgrade_TextBlock : text_block = text_block{DefaultTextColor:=UpgradeTextColor, DefaultShadowOffset:=ShadowOffset, DefaultShadowOpacity := DefaultTextShadowOpacity}
    UP_Row5_Upgrade_TextBlock : text_block = text_block{DefaultTextColor:=UpgradeTextColor, DefaultShadowOffset:=ShadowOffset, DefaultShadowOpacity := DefaultTextShadowOpacity}

    UP_Row1_UpgradeCost_TextBlock : text_block = text_block{DefaultTextColor:=SapphiresTextColor, DefaultShadowOffset:=ShadowOffset, DefaultShadowOpacity := DefaultTextShadowOpacity}
    UP_Row2_UpgradeCost_TextBlock : text_block = text_block{DefaultTextColor:=SapphiresTextColor, DefaultShadowOffset:=ShadowOffset, DefaultShadowOpacity := DefaultTextShadowOpacity}
    UP_Row3_UpgradeCost_TextBlock : text_block = text_block{DefaultTextColor:=SapphiresTextColor, DefaultShadowOffset:=ShadowOffset, DefaultShadowOpacity := DefaultTextShadowOpacity}
    UP_Row4_UpgradeCost_TextBlock : text_block = text_block{DefaultTextColor:=SapphiresTextColor, DefaultShadowOffset:=ShadowOffset, DefaultShadowOpacity := DefaultTextShadowOpacity}
    UP_Row5_UpgradeCost_TextBlock : text_block = text_block{DefaultTextColor:=SapphiresTextColor, DefaultShadowOffset:=ShadowOffset, DefaultShadowOpacity := DefaultTextShadowOpacity}

    UP_NumPage_TextBlock : text_block = text_block{DefaultTextColor:=White, DefaultShadowOffset:=ShadowOffset, DefaultShadowOpacity := 0.0}
    
    var AllUP_Rows_Tier_TextBlock : []text_block = array{}
    var AllUP_Rows_Level_TextBlock : []text_block = array{}
    var AllUP_Rows_Production_TextBlock : []text_block = array{}
    var AllUP_Rows_Upgrade_TextBlock : []text_block = array{}
    var AllUP_Rows_UpgradeCost_TextBlock : []text_block = array{}
    var AllUP_Rows_SapphiresTexture : []texture_block = array{}
    var AllUP_Rows_TextureBlock : []texture_block = array{}

#Baker speed resourcers efficiency

    SREColumn1SizeX : float = 150.0
    SREColumn1SizeY : float = 80.0

    SREColumn2SizeX : float = 200.0
    SREColumn2SizeY : float = 80.0

    SREColumn3SizeX : float = 150.0
    SREColumn3SizeY : float = 80.0

    SREColumn4SizeX : float = 225.0
    SREColumn4SizeY : float = 80.0

    #Data
    BS_RowButton : button_quiet = button_quiet{}

    RE_Row1Button : button_quiet = button_quiet{}
    RE_Row2Button : button_quiet = button_quiet{}

    BS_Row_Level_IconTexture : texture_block = texture_block{DefaultDesiredSize:=vector2{X:=SizeTextureBlockX,Y:=SizeTextureBlockY},DefaultImage:=Images.EmptyTexture}
    BS_Row_Speed_TextBlock : text_block = text_block{DefaultTextColor:=CurrentTextColor, DefaultShadowOffset:=ShadowOffset, DefaultShadowOpacity := DefaultTextShadowOpacity}
    BS_Row_Upgrade_TextBlock : text_block = text_block{DefaultTextColor:=UpgradeTextColor, DefaultShadowOffset:=ShadowOffset, DefaultShadowOpacity := DefaultTextShadowOpacity}
    BS_Row_UpgradeCost_TextBlock : text_block = text_block{DefaultTextColor:=SapphiresTextColor, DefaultShadowOffset:=ShadowOffset, DefaultShadowOpacity := DefaultTextShadowOpacity}
    BS_Row_SapphiresTexture : texture_block = texture_block{DefaultDesiredSize:=vector2{X:=SizeSmallIconX,Y:=SizeSmallIconY},DefaultImage:=Images.Sapphire_Icon}

    RE_Row1_IconTexture : texture_block = texture_block{DefaultDesiredSize:=vector2{X:=SizeTextureBlockX,Y:=SizeTextureBlockY},DefaultImage:=Images.EmptyTexture}
    RE_Row2_IconTexture : texture_block = texture_block{DefaultDesiredSize:=vector2{X:=SizeTextureBlockX,Y:=SizeTextureBlockY},DefaultImage:=Images.EmptyTexture}

    RE_Row1_LineTexture : texture_block = texture_block{DefaultImage:=Images.EmptyTexture}
    RE_Row2_LineTexture : texture_block = texture_block{DefaultImage:=Images.EmptyTexture}

    RE_Row1_SapphiresTexture : texture_block = texture_block{DefaultDesiredSize:=vector2{X:=SizeSmallIconX,Y:=SizeSmallIconY},DefaultImage:=Images.Sapphire_Icon}
    RE_Row2_SapphiresTexture : texture_block = texture_block{DefaultDesiredSize:=vector2{X:=SizeSmallIconX,Y:=SizeSmallIconY},DefaultImage:=Images.Sapphire_Icon}

    RE_Row1_Efficiency_TextBlock : text_block = text_block{DefaultTextColor:=CurrentTextColor, DefaultShadowOffset:=ShadowOffset, DefaultShadowOpacity := DefaultTextShadowOpacity}
    RE_Row2_Efficiency_TextBlock : text_block = text_block{DefaultTextColor:=CurrentTextColor, DefaultShadowOffset:=ShadowOffset, DefaultShadowOpacity := DefaultTextShadowOpacity}

    RE_Row1_Upgrade_TextBlock : text_block = text_block{DefaultTextColor:=UpgradeTextColor, DefaultShadowOffset:=ShadowOffset, DefaultShadowOpacity := DefaultTextShadowOpacity}
    RE_Row2_Upgrade_TextBlock : text_block = text_block{DefaultTextColor:=UpgradeTextColor, DefaultShadowOffset:=ShadowOffset, DefaultShadowOpacity := DefaultTextShadowOpacity}

    RE_Row1_UpgradeCost_TextBlock : text_block = text_block{DefaultTextColor:=SapphiresTextColor, DefaultShadowOffset:=ShadowOffset, DefaultShadowOpacity := DefaultTextShadowOpacity}
    RE_Row2_UpgradeCost_TextBlock : text_block = text_block{DefaultTextColor:=SapphiresTextColor, DefaultShadowOffset:=ShadowOffset, DefaultShadowOpacity := DefaultTextShadowOpacity}
    

#Top overlay
    TotalProduction_TextBlock : text_block = text_block{DefaultTextColor:=CurrentTextColor, DefaultShadowOffset:=ShadowOffset, DefaultShadowOpacity := DefaultTextShadowOpacity}
    TotalSapphires_TextBlock : text_block = text_block{DefaultTextColor:=SapphiresTextColor, DefaultShadowOffset:=ShadowOffset, DefaultShadowOpacity := DefaultTextShadowOpacity}

    CurrentBuyOption_textblock : text_block = text_block{DefaultTextColor:=White}

    BuyOption_buttonquiet : button_quiet = button_quiet{}

    var MaybeCanvas : ?canvas = false
    var UIIsDisplayed : logic = false

    var TimeElapsed : int = 0

    var CurrentPage : int = 1
    var BuyModeMultiplier : int = 1

    RunWarning(Amount : int)<suspends> : void =
        WarningTitle_TextBlock.SetText(BasicMessage("WARNING: Improve your pickaxe damage first"))
        Warning_TextBlock.SetText(BasicMessage("This production requires x{Amount} pickaxe hits"))
        Sleep(5.0)
        WarningTitle_TextBlock.SetText(BasicMessage(""))
        Warning_TextBlock.SetText(BasicMessage(""))

    InitUI() : void =
        SetArrays()
        InitSubscribers()
        Canvas := MakeCanvas()
        set MaybeCanvas = option{Canvas}
        RemoveDelay_OnRemove()

    ChangeBuyMode(WM : widget_message) : void =
        if(BakeryMachinesManager.PlayerDataCache.Inventory_Fame>1000.0):
            case(BuyModeMultiplier):
                1=> set BuyModeMultiplier = 5
                5=> set BuyModeMultiplier = 25
                25=> set BuyModeMultiplier = 1
                _=>#Print("ChangeBuyMode not valid")
            BuyOption_buttonquiet.SetText(BasicMessage("x{BuyModeMultiplier}"))
            BakeryMachinesManager.RefreshBProductionUI(BB_TYPE)

    CheckBuyOptionButton() : void =
        if(BakeryMachinesManager.PlayerDataCache.Inventory_Fame>1000.0):
            BuyOption_buttonquiet.SetText(BasicMessage("x{BuyModeMultiplier}"))
        else:
            BuyOption_buttonquiet.SetText(BasicMessage("Locked"))




    SetArrays() : void =
        set AllUP_Rows_Tier_TextBlock = array{UP_Row1_Tier_TextBlock,UP_Row2_Tier_TextBlock,UP_Row3_Tier_TextBlock,UP_Row4_Tier_TextBlock,UP_Row5_Tier_TextBlock}
        set AllUP_Rows_Level_TextBlock = array{UP_Row1_Level_TextBlock,UP_Row2_Level_TextBlock,UP_Row3_Level_TextBlock,UP_Row4_Level_TextBlock,UP_Row5_Level_TextBlock}
        set AllUP_Rows_Production_TextBlock = array{UP_Row1_Production_TextBlock,UP_Row2_Production_TextBlock,UP_Row3_Production_TextBlock,UP_Row4_Production_TextBlock,UP_Row5_Production_TextBlock}
        set AllUP_Rows_Upgrade_TextBlock = array{UP_Row1_Upgrade_TextBlock,UP_Row2_Upgrade_TextBlock,UP_Row3_Upgrade_TextBlock,UP_Row4_Upgrade_TextBlock,UP_Row5_Upgrade_TextBlock}
        set AllUP_Rows_UpgradeCost_TextBlock = array{UP_Row1_UpgradeCost_TextBlock,UP_Row2_UpgradeCost_TextBlock,UP_Row3_UpgradeCost_TextBlock,UP_Row4_UpgradeCost_TextBlock,UP_Row5_UpgradeCost_TextBlock}
        set AllUP_Rows_SapphiresTexture = array{UP_Row1_SapphiresTexture,UP_Row2_SapphiresTexture,UP_Row3_SapphiresTexture,UP_Row4_SapphiresTexture,UP_Row5_SapphiresTexture}
        set AllUP_Rows_TextureBlock = array{UPRow1TextureBlock,UPRow2TextureBlock,UPRow3TextureBlock,UPRow4TextureBlock,UPRow5TextureBlock}
    
    InitSubscribers() : void =
        CloseButton.OnClick().Subscribe(Remove_Canvas)
        BuyOption_buttonquiet.OnClick().Subscribe(ChangeBuyMode)

        UP_Row1Button.OnClick().Subscribe(handler_upgrades{MaybeBakeryMachinesManager := option{BakeryMachinesManager}, Row:=0, BakerType := BB_TYPE}.HandlerUpgradeTier)  
        UP_Row2Button.OnClick().Subscribe(handler_upgrades{MaybeBakeryMachinesManager := option{BakeryMachinesManager}, Row:=1, BakerType := BB_TYPE}.HandlerUpgradeTier) 
        UP_Row3Button.OnClick().Subscribe(handler_upgrades{MaybeBakeryMachinesManager := option{BakeryMachinesManager}, Row:=2, BakerType := BB_TYPE}.HandlerUpgradeTier) 
        UP_Row4Button.OnClick().Subscribe(handler_upgrades{MaybeBakeryMachinesManager := option{BakeryMachinesManager}, Row:=3, BakerType := BB_TYPE}.HandlerUpgradeTier)  
        UP_Row5Button.OnClick().Subscribe(handler_upgrades{MaybeBakeryMachinesManager := option{BakeryMachinesManager}, Row:=4, BakerType := BB_TYPE}.HandlerUpgradeTier) 

        UP_PrevPageButton.OnClick().Subscribe(SelectPrevPage) 
        UP_NextPageButton.OnClick().Subscribe(SelectNextPage)

        BS_RowButton.OnClick().Subscribe(handler_upgrades{MaybeBakeryMachinesManager := option{BakeryMachinesManager}, Row:=0, BakerType := BB_TYPE}.HandlerUpgradeSpeed)

        RE_Row1Button.OnClick().Subscribe(handler_upgrades{MaybeBakeryMachinesManager := option{BakeryMachinesManager}, Row:=0, BakerType := BB_TYPE}.HandlerUpgradeEfficiency)
        RE_Row2Button.OnClick().Subscribe(handler_upgrades{MaybeBakeryMachinesManager := option{BakeryMachinesManager}, Row:=1, BakerType := BB_TYPE}.HandlerUpgradeEfficiency)  
    
    SelectPrevPage(WM : widget_message) : void =
        if(CurrentPage>1):
            set CurrentPage = CurrentPage - 1
        else:
            set CurrentPage = 4
        BakeryMachinesManager.RefreshBProductionUI(BB_TYPE)

    SelectNextPage(WM : widget_message) : void =
        if(CurrentPage<4):
            set CurrentPage = CurrentPage + 1
        else:
            set CurrentPage = 1
        BakeryMachinesManager.RefreshBProductionUI(BB_TYPE)

    SetAgent(InAgent : agent) : void =
        set CurrentPage = 1
        set BuyModeMultiplier = 1
        set MaybeAgent = option{InAgent}

    UpdateUI(PlayerData : player_data_table, TotalProduction : float) : void =
        CheckBuyOptionButton()

        TotalSapphires_TextBlock.SetText(BasicMessage(GetMagnitudeLetter(PlayerData.Inventory_Sapphires,"", "")))
        Log("TotalProduction: {TotalProduction}",LOG_LEVEL_UIs)
        TotalProduction_TextBlock.SetText(BasicMessage("Total production: " + GetMagnitudeLetter(TotalProduction,"", "")))

        UP_Row1_Tier_TextBlock.SetText(BasicMessage("{1*CurrentPage}"))
        UP_Row2_Tier_TextBlock.SetText(BasicMessage("{2*CurrentPage}"))
        UP_Row3_Tier_TextBlock.SetText(BasicMessage("{3*CurrentPage}"))
        UP_Row4_Tier_TextBlock.SetText(BasicMessage("{4*CurrentPage}"))
        UP_Row5_Tier_TextBlock.SetText(BasicMessage("{5*CurrentPage}"))

        var RowIndex : int = 0
        #AllUP_Rows_Level_TextBlock
        for(TierSelected:=(CurrentPage-1)*5..((CurrentPage-1)*5)+5-1,OriginalTierLevel:=PlayerData.TLV_BB[TierSelected]):
            var TierLevel : int = OriginalTierLevel
            if(TierLevel>MaxTierLevel):
                set TierLevel = MaxTierLevel
            if:
                TierTextBlock:=AllUP_Rows_Tier_TextBlock[RowIndex]
                LevelTextBlock:=AllUP_Rows_Level_TextBlock[RowIndex]
                ProductionTextBlock:=AllUP_Rows_Production_TextBlock[RowIndex]
                UpgradeTextBlock:=AllUP_Rows_Upgrade_TextBlock[RowIndex]
                UpgradeCostTextBlock:=AllUP_Rows_UpgradeCost_TextBlock[RowIndex]
                UpgradeCostTexture:=AllUP_Rows_SapphiresTexture[RowIndex]
                UpgradeIcon:=AllUP_Rows_TextureBlock[RowIndex]
                TierTintColor := AllTierColors[TierSelected]
            then:
                TierTextBlock.SetText(BasicMessage("{TierSelected}"))
                LevelTextBlock.SetText(BasicMessage("{TierLevel}"))
                var CurrentValue : float = 0.0 
                var Upgrade : float = 0.0 
                var UpgradeDifference : float = 0.0 
                var UpgradeCost : float = 0.0 
                if(TierLevel<>-1,TierLevel<MaxTierLevel):
                    TierTextBlock.SetText(BasicMessage("{TierSelected+1}"))
                    LevelTextBlock.SetText(BasicMessage("{TierLevel}"))
                    CompleteTuple := CalculeBakerUpgradeProduction(TierSelected,TierLevel,BB_TYPE,BuyModeMultiplier)
                    set CurrentValue = CompleteTuple(0)
                    set UpgradeCost = CompleteTuple(1)
                    set UpgradeDifference = CompleteTuple(2)
                    ProductionTextBlock.SetText(BasicMessage(GetMagnitudeLetter(CurrentValue,"", "")))
                    ProductionTextBlock.SetTextColor(CurrentTextColor)
                    UpgradeTextBlock.SetText(BasicMessage(GetMagnitudeLetter(UpgradeDifference,"+", "")))
                    UpgradeTextBlock.SetTextColor(UpgradeTextColor)
                    UpgradeCostTextBlock.SetText(BasicMessage(GetMagnitudeLetter(UpgradeCost,"x ", "")))
                    UpgradeCostTextBlock.SetTextColor(SapphiresTextColor)
                    UpgradeCostTexture.SetImage(Images.Sapphire_Icon)
                    UpgradeIcon.SetImage(Images.Pipes_Icon)
                    if(PlayerData.Inventory_Sapphires>=UpgradeCost):
                        UpgradeCostTextBlock.SetTextColor(SapphiresTextColor)
                        UpgradeIcon.SetTint(TierTintColor)
                    else:
                        UpgradeCostTextBlock.SetTextColor(NotEnoughTextColor)
                        UpgradeIcon.SetTint(White)
                else if(TierLevel>=MaxTierLevel):
                    CompleteTuple := CalculeBakerUpgradeProduction(TierSelected,TierLevel,BB_TYPE,BuyModeMultiplier)
                    set CurrentValue = CompleteTuple(0)
                    TierTextBlock.SetText(BasicMessage("{TierSelected+1}"))
                    LevelTextBlock.SetText(BasicMessage("{TierLevel}"))
                    ProductionTextBlock.SetText(BasicMessage(GetMagnitudeLetter(CurrentValue,"", "")))
                    ProductionTextBlock.SetTextColor(CurrentTextColor)
                    UpgradeTextBlock.SetText(BasicMessage("MAXED"))
                    UpgradeTextBlock.SetTextColor(MaxedColor)
                    UpgradeCostTextBlock.SetText(BasicMessage("MAXED"))
                    UpgradeCostTextBlock.SetTextColor(MaxedColor)
                    UpgradeCostTexture.SetImage(Images.Sapphire_Icon)
                    UpgradeIcon.SetImage(Images.Pipes_Icon)
                    <#
                    if(PlayerData.Inventory_Sapphires>=UpgradeCost):
                        UpgradeCostTextBlock.SetTextColor(SapphiresTextColor)
                        UpgradeIcon.SetTint(TierTintColor)
                    else:
                        UpgradeCostTextBlock.SetTextColor(NotEnoughTextColor)
                        UpgradeIcon.SetTint(White)#>
                else:
                    TierTextBlock.SetText(BasicMessage(""))
                    LevelTextBlock.SetText(BasicMessage(""))
                    ProductionTextBlock.SetText(BasicMessage("LOCKED"))
                    ProductionTextBlock.SetTextColor(NotEnoughTextColor)
                    UpgradeTextBlock.SetText(BasicMessage("LOCKED"))
                    UpgradeTextBlock.SetTextColor(NotEnoughTextColor)
                    UpgradeCostTextBlock.SetText(BasicMessage("LOCKED"))
                    UpgradeCostTextBlock.SetTextColor(NotEnoughTextColor)
                    UpgradeCostTexture.SetImage(Images.Cross_Icon)
                    UpgradeIcon.SetTint(White)
                    UpgradeIcon.SetImage(Images.Cross_Icon)
                if(BakeryMachinesManager.IsOnUKTutorial?, TierSelected = 0):
                    ProductionTextBlock.SetText(BasicMessage(GetMagnitudeLetter(CurrentValue,"", "")))
                    ProductionTextBlock.SetTextColor(CurrentTextColor)
                    UpgradeTextBlock.SetText(BasicMessage(GetMagnitudeLetter(UpgradeDifference,"+", "")))
                    UpgradeTextBlock.SetTextColor(UpgradeTextColor)
                    UpgradeCostTextBlock.SetText(BasicMessage(GetMagnitudeLetter(0.0,"x ", "")))
                    UpgradeCostTextBlock.SetTextColor(SapphiresTextColor)
                    UpgradeIcon.SetTint(TierTintColor)
            set RowIndex = RowIndex + 1

        UP_PrevPageButton.SetText(BasicMessage("👈 Prev. Page"))
        UP_NextPageButton.SetText(BasicMessage("Next Page 👉"))
        CloseButton.SetText(BasicMessage("Close"))

        UP_NumPage_TextBlock.SetText(BasicMessage("{CurrentPage}/{4} Page"))
      
        #Speed
        LV_BBSpeed := PlayerData.LV_BBSpeed
        BS_Row_Level_IconTexture.SetImage(Images.Speed_Icon)
        MaxLevel := CalculeSpeedLimitLevel(Base_BB_Speed)
        SpeedTuple :=CalculeBakerSpeed(BB_TYPE, LV_BBSpeed, BuyModeMultiplier)

        BS_Row_Speed_TextBlock.SetText(BasicMessage(GetMagnitudeLetter(SpeedTuple(0),"", " s")))
        if(LV_BBSpeed<MaxLevel):
            BS_Row_Upgrade_TextBlock.SetText(BasicMessage(GetMagnitudeLetter(SpeedTuple(2),"-", " s")))
            if(PlayerData.Inventory_Sapphires>=SpeedTuple(1)):
                BS_Row_UpgradeCost_TextBlock.SetTextColor(SapphiresTextColor)
            else:
                BS_Row_UpgradeCost_TextBlock.SetTextColor(NotEnoughTextColor)
            BS_Row_UpgradeCost_TextBlock.SetText(BasicMessage(GetMagnitudeLetter(SpeedTuple(1),"x ", "")))
        else:
            BS_Row_Upgrade_TextBlock.SetText(BasicMessage("MAXED"))
            BS_Row_Upgrade_TextBlock.SetTextColor(MaxedColor)
            BS_Row_UpgradeCost_TextBlock.SetText(BasicMessage("MAXED"))
            BS_Row_UpgradeCost_TextBlock.SetTextColor(MaxedColor)

        if:
            WaterEfficiencyLevel := PlayerData.EfficiencyLV_BB[0]
            FlourEfficiencyLevel := PlayerData.EfficiencyLV_BB[1]
        then:
            WaterEfficiencyTuple:=CalculeBakerResourceEfficiency(WaterEfficiencyLevel, BB_TYPE, WATER, BuyModeMultiplier)
            FlourEfficiencyTuple:=CalculeBakerResourceEfficiency(FlourEfficiencyLevel, BB_TYPE, FLOUR, BuyModeMultiplier)

            RE_Row1_Efficiency_TextBlock.SetText(BasicMessage(GetMagnitudeLetter(WaterEfficiencyTuple(0)+100.0,"", "%")))
            RE_Row2_Efficiency_TextBlock.SetText(BasicMessage(GetMagnitudeLetter(FlourEfficiencyTuple(0)+100.0,"", "%")))

            RE_Row1_UpgradeCost_TextBlock.SetText(BasicMessage(GetMagnitudeLetter(WaterEfficiencyTuple(1),"x ", "")))
            if(PlayerData.Inventory_Sapphires>=WaterEfficiencyTuple(1)):
                RE_Row1_UpgradeCost_TextBlock.SetTextColor(SapphiresTextColor)
            else:
                RE_Row1_UpgradeCost_TextBlock.SetTextColor(NotEnoughTextColor)
            RE_Row2_UpgradeCost_TextBlock.SetText(BasicMessage(GetMagnitudeLetter(FlourEfficiencyTuple(1),"x ", "")))
            if(PlayerData.Inventory_Sapphires>=FlourEfficiencyTuple(1)):
                RE_Row2_UpgradeCost_TextBlock.SetTextColor(SapphiresTextColor)
            else:
                RE_Row2_UpgradeCost_TextBlock.SetTextColor(NotEnoughTextColor)

            if(WaterEfficiencyLevel<MaxBakerEfficiencyLV):
                RE_Row1_Upgrade_TextBlock.SetText(BasicMessage(GetMagnitudeLetter(WaterEfficiencyTuple(2),"+", "%")))
                RE_Row1_Upgrade_TextBlock.SetTextColor(UpgradeTextColor)
            else:
                RE_Row1_Upgrade_TextBlock.SetText(BasicMessage("MAXED"))
                RE_Row1_Upgrade_TextBlock.SetTextColor(MaxedColor)
                RE_Row1_UpgradeCost_TextBlock.SetText(BasicMessage("MAXED"))
                RE_Row1_UpgradeCost_TextBlock.SetTextColor(MaxedColor)

            if(FlourEfficiencyLevel<MaxBakerEfficiencyLV):
                RE_Row2_Upgrade_TextBlock.SetText(BasicMessage(GetMagnitudeLetter(FlourEfficiencyTuple(2),"+", "%")))
                RE_Row2_Upgrade_TextBlock.SetTextColor(UpgradeTextColor)
            else:
                RE_Row2_Upgrade_TextBlock.SetText(BasicMessage("MAXED"))
                RE_Row2_Upgrade_TextBlock.SetTextColor(MaxedColor)
                RE_Row2_UpgradeCost_TextBlock.SetText(BasicMessage("MAXED"))
                RE_Row2_UpgradeCost_TextBlock.SetTextColor(MaxedColor)

            
            RE_Row1_SapphiresTexture.SetImage(Images.Sapphire_Icon)
            RE_Row2_SapphiresTexture.SetImage(Images.Sapphire_Icon)

            #Only display the necesary Resourcers
            RE_Row1_IconTexture.SetImage(Images.Water_Icon)
            RE_Row2_IconTexture.SetImage(Images.Flour_Icon)

            #Only display the necesary lines
            RE_Row1_LineTexture.SetImage(Images.Img_LineBackground)
            RE_Row1_LineTexture.SetDesiredSize(vector2{X:=0.0,Y:=LineSizeY})
            RE_Row2_LineTexture.SetImage(Images.Img_LineBackground)
            RE_Row2_LineTexture.SetDesiredSize(vector2{X:=0.0,Y:=LineSizeY})

    Remove_Canvas(Message : widget_message) : void =
        set UIIsDisplayed = false
        if:
            Agent:=MaybeAgent?
            Player:=player[Agent]
            PlayerUI:=GetPlayerUI[Player]
            Canvas := MaybeCanvas?
        then:
            PlayerUI.RemoveWidget(Canvas)
            RemoveDelay_OnRemove()

    Remove_CanvasWithoutMessage() : void =
        set UIIsDisplayed = false
        #Print("Trying to remove UI")
        if:
            Agent:=MaybeAgent?
            Player:=player[Agent]
            PlayerUI:=GetPlayerUI[Player]
            Canvas := MaybeCanvas?
        then:
            #Print("Removing UI")
            PlayerUI.RemoveWidget(Canvas)
            RemoveDelay_OnRemove()

            
    Display_Canvas() : void =
        set BakeryMachinesManager.WarningMessageDisplayed = false
        set TimeElapsed = 0
        set UIIsDisplayed = true
        spawn: 
            StartTimeElapsed()
        if:
            Agent:=MaybeAgent?
            PlayerUI:=GetPlayerUI[player[Agent]]
            Canvas := MaybeCanvas?
        then:
            ##Print("Assigning created UI before")
            PlayerUI.AddWidget(Canvas,player_ui_slot{InputMode := ui_input_mode.All, ZOrder:=1})
            RemoveDelay_OnAdd()   

    StartTimeElapsed()<suspends> : void =
        loop:
            ##Print("IN THE LOOP TimeElapsed: {TimeElapsed}")
            if(TimeElapsed > MaxUIDisplayTime):
                ##Print("Closing the UI because 20 seconds has pased")
                Remove_CanvasWithoutMessage()
                break
            if(UIIsDisplayed = false):
                ##Print("Breaking Time Elapsed loop")
                break
            Sleep(1.0)
            set TimeElapsed = TimeElapsed + 1

    RemoveDelay_OnAdd() : void =
        CurrentBuyOption_textblock.SetText(BasicMessage("Buy multiplier:"))
        #CurrentBuyOption_textblock.SetVisibility(widget_visibility.Hidden)

    RemoveDelay_OnRemove() : void =
        CurrentBuyOption_textblock.SetText(BasicMessage("Buy multiplier: "))
        #CurrentBuyOption_textblock.SetVisibility(widget_visibility.Hidden)

    <#Offsets
    # 
    When the anchors are well defined:
        Offsets.Left represent the distance in pixels from the Anchors Minimum.X
        Offsets.Bottom represent the distance in pixel from the Anchors Maximum.Y, effectively controlling the desired widget size. 
    When the anchors are not well defined, 
        Offsets.Left and Offsets.Top represent the widget position
        Offsets.Right and Offset.Bottom represent the widget size.
    #>
    MakeCanvas(): canvas=
        Canvas := canvas:
            Slots:=array:
                #Background black image
                canvas_slot:
                    Anchors:=anchors{Minimum:=vector2{X:=0.5,Y:=0.5},Maximum:=vector2{X:=0.5,Y:=0.5}}
                    ZOrder:=0
                    Alignment:=vector2{X:=0.5,Y:=0.5}
                    Widget:=texture_block:
                        DefaultImage:=Images.BlackBackground
                        DefaultDesiredSize:=vector2{X:=BackgroundBlackUISizeX,Y:=BackgroundBlackUISizeY}
                #Background image
                canvas_slot:
                    Anchors:=anchors{Minimum:=vector2{X:=0.5,Y:=0.5},Maximum:=vector2{X:=0.5,Y:=0.5}}
                    ZOrder:=0
                    Offsets:=margin{Left:=0.0,Top:=0.0,Right:=BackgroundImageSizeX,Bottom:=BackgroundImageSizeY} #PositionX, PositionY, SizeX, SizeY
                    Alignment:=vector2{X:=0.5,Y:=0.5}
                    Widget:=texture_block:
                        DefaultImage:=Images.Img_MainBackground
                        DefaultDesiredSize:=vector2{X:=BackgroundImageSizeX,Y:=BackgroundImageSizeY}
                #Main overlay
                canvas_slot:
                    Anchors:=anchors{Minimum:=vector2{X:=0.5,Y:=0.5},Maximum:=vector2{X:=0.5,Y:=0.5}}
                    ZOrder:=0
                    #Offsets:=margin{Left:=1920.0,Top:=0.0,Right:=0.0,Bottom:=1080.0} #PositionX, PositionY, SizeX, SizeY
                    Alignment:=vector2{X:=0.5,Y:=0.5}
                        Widget:=stack_box:
                            Orientation:=orientation.Vertical
                                Slots:=array:
                                    stack_box_slot:
                                        HorizontalAlignment:=horizontal_alignment.Fill
                                        VerticalAlignment:=vertical_alignment.Center
                                        Padding:=margin{Top:=20.0}
                                        Widget:=stack_box:
                                            Orientation:=orientation.Horizontal
                                                Slots:=array:
                                                    stack_box_slot:
                                                        HorizontalAlignment:=horizontal_alignment.Fill
                                                        VerticalAlignment:=vertical_alignment.Fill
                                                        #Upgrade level stack box
                                                        Widget:=stack_box:
                                                            Orientation:=orientation.Vertical
                                                                Slots:=array:   
                                                                    stack_box_slot: #Line -1
                                                                        HorizontalAlignment:=horizontal_alignment.Center
                                                                        VerticalAlignment:=vertical_alignment.Center
                                                                        Padding:=margin{Left:=RowLeftPadding,Top:=RowTopPadding,Right:=RowRightPadding,Bottom:=RowBottomPadding}
                                                                        Widget:=overlay:
                                                                            Slots:=array:#Row 1
                                                                                overlay_slot: #Line background
                                                                                    HorizontalAlignment:=horizontal_alignment.Center
                                                                                    VerticalAlignment:=vertical_alignment.Center
                                                                                    Widget:=texture_block{DefaultDesiredSize:=vector2{X:=LineSizeX,Y:=LineSizeY},DefaultImage:=Images.EmptyTexture}
                                                                                overlay_slot: #Content
                                                                                    Widget:=stack_box:
                                                                                        Orientation:=orientation.Horizontal
                                                                                            Slots:=array:
                                                                                                stack_box_slot: #Column 1
                                                                                                    Padding:=margin{Left:=InterColumnLeftPadding,Top:=InterColumnTopPadding,Right:=InterColumnRightPadding,Bottom:=InterColumnBottomPadding}
                                                                                                    Widget:=overlay:
                                                                                                        Slots:=array:
                                                                                                            overlay_slot: 
                                                                                                                Widget:=texture_block{DefaultDesiredSize:=vector2{X:=UPColumn1SizeX,Y:=UPColumn1SizeY},DefaultImage:=Images.EmptyTexture}
                                                                                                stack_box_slot: #Column 2
                                                                                                    Padding:=margin{Left:=InterColumnLeftPadding,Top:=InterColumnTopPadding,Right:=InterColumnRightPadding,Bottom:=InterColumnBottomPadding}
                                                                                                    Widget:=overlay:
                                                                                                        Slots:=array:
                                                                                                            overlay_slot: 
                                                                                                                Widget:=texture_block{DefaultDesiredSize:=vector2{X:=UPColumn2SizeX,Y:=UPColumn2SizeY},DefaultImage:=Images.EmptyTexture}
                                                                                                stack_box_slot: #Column 3
                                                                                                    Padding:=margin{Left:=InterColumnLeftPadding,Top:=InterColumnTopPadding,Right:=InterColumnRightPadding,Bottom:=InterColumnBottomPadding}
                                                                                                    Widget:=overlay:
                                                                                                        Slots:=array:
                                                                                                            overlay_slot: 
                                                                                                                Widget:=texture_block{DefaultDesiredSize:=vector2{X:=UPColumn3SizeX,Y:=UPColumn3SizeY},DefaultImage:=Images.EmptyTexture}
                                                                                                stack_box_slot: #Column 4
                                                                                                    Padding:=margin{Left:=InterColumnLeftPadding,Top:=InterColumnTopPadding,Right:=InterColumnRightPadding,Bottom:=InterColumnBottomPadding}
                                                                                                    Widget:=overlay:
                                                                                                        Slots:=array:
                                                                                                            overlay_slot: 
                                                                                                                Widget:=texture_block{DefaultDesiredSize:=vector2{X:=UPColumn4SizeX,Y:=UPColumn4SizeY},DefaultImage:=Images.EmptyTexture}
                                                                                                stack_box_slot: #Column 5
                                                                                                    Padding:=margin{Left:=InterColumnLeftPadding,Top:=InterColumnTopPadding,Right:=InterColumnRightPadding,Bottom:=InterColumnBottomPadding}
                                                                                                    Widget:=overlay:
                                                                                                        Slots:=array:
                                                                                                            overlay_slot:
                                                                                                                Widget:=overlay:
                                                                                                                    Slots:=array:
                                                                                                                        overlay_slot:
                                                                                                                            Widget:=texture_block{DefaultDesiredSize:=vector2{X:=UPColumn5SizeX,Y:=UPColumn5SizeY},DefaultImage:=Images.EmptyTexture}       
                                                                                                                        overlay_slot:
                                                                                                                            HorizontalAlignment:=horizontal_alignment.Right
                                                                                                                            VerticalAlignment:=vertical_alignment.Center
                                                                                                                            Widget:=CurrentBuyOption_textblock
                                                                                                stack_box_slot: #Column 6
                                                                                                    Padding:=margin{Left:=InterColumnLeftPadding,Top:=InterColumnTopPadding,Right:=InterColumnRightPadding,Bottom:=InterColumnBottomPadding}
                                                                                                    Widget:=overlay:
                                                                                                        Slots:=array:
                                                                                                            overlay_slot: 
                                                                                                                Widget:=texture_block{DefaultDesiredSize:=vector2{X:=UPColumn6SizeX,Y:=UPColumn6SizeY},DefaultImage:=Images.EmptyTexture}
                                                                                                            overlay_slot:
                                                                                                                VerticalAlignment:=vertical_alignment.Fill  
                                                                                                                Widget:=BuyOption_buttonquiet
                                                                    stack_box_slot: #Line 0
                                                                        HorizontalAlignment:=horizontal_alignment.Center
                                                                        VerticalAlignment:=vertical_alignment.Center
                                                                        Padding:=margin{Left:=RowLeftPadding,Top:=RowTopPadding,Right:=RowRightPadding,Bottom:=RowBottomPadding}
                                                                        Widget:=overlay:
                                                                            Slots:=array:#Row 1
                                                                                overlay_slot:
                                                                                    Widget:=texture_block{DefaultDesiredSize:=vector2{X:=490.0,Y:=80.0},DefaultImage:=Images.EmptyTexture} 
                                                                                overlay_slot:
                                                                                    HorizontalAlignment:=horizontal_alignment.Fill
                                                                                    VerticalAlignment:=vertical_alignment.Fill
                                                                                    Widget:=texture_block{DefaultImage:=Images.Img_LineBackground2}   
                                                                                overlay_slot:
                                                                                    HorizontalAlignment:=horizontal_alignment.Left
                                                                                    VerticalAlignment:=vertical_alignment.Center
                                                                                    Padding:=margin{Left:=Icon_LeftMargin}
                                                                                    Widget:=texture_block{DefaultDesiredSize:=vector2{X:=IconSmallSizeX,Y:=IconSmallSizeY},DefaultImage:=Images.Flour_Icon}
                                                                                overlay_slot:
                                                                                    HorizontalAlignment:=horizontal_alignment.Left
                                                                                    VerticalAlignment:=vertical_alignment.Center
                                                                                    Padding:=margin{Left:=TextIcon_LeftMargin}
                                                                                    Widget:=TotalProduction_TextBlock
                                                                    stack_box_slot: #Line 1
                                                                        HorizontalAlignment:=horizontal_alignment.Center
                                                                        VerticalAlignment:=vertical_alignment.Center
                                                                        Padding:=margin{Left:=RowLeftPadding,Top:=RowTopPadding,Right:=RowRightPadding,Bottom:=RowBottomPadding}
                                                                        Widget:=overlay:
                                                                            Slots:=array:#Row 1
                                                                                overlay_slot: #Line background
                                                                                    HorizontalAlignment:=horizontal_alignment.Center
                                                                                    VerticalAlignment:=vertical_alignment.Center
                                                                                    Widget:=texture_block{DefaultDesiredSize:=vector2{X:=LineSizeX,Y:=LineSizeY},DefaultImage:=Images.Img_LineBackground}
                                                                                overlay_slot: #Content
                                                                                    Widget:=stack_box:
                                                                                        Orientation:=orientation.Horizontal
                                                                                            Slots:=array:
                                                                                                stack_box_slot: #Column 1
                                                                                                    Padding:=margin{Left:=InterColumnLeftPadding,Top:=InterColumnTopPadding,Right:=InterColumnRightPadding,Bottom:=InterColumnBottomPadding}
                                                                                                    Widget:=overlay:
                                                                                                        Slots:=array:
                                                                                                            overlay_slot: 
                                                                                                                Widget:=texture_block{DefaultDesiredSize:=vector2{X:=UPColumn1SizeX,Y:=UPColumn1SizeY},DefaultImage:=Images.EmptyTexture}
                                                                                                stack_box_slot: #Column 2
                                                                                                    Padding:=margin{Left:=InterColumnLeftPadding,Top:=InterColumnTopPadding,Right:=InterColumnRightPadding,Bottom:=InterColumnBottomPadding}
                                                                                                    Widget:=overlay:
                                                                                                        Slots:=array:
                                                                                                            overlay_slot: 
                                                                                                                Widget:=text_block{DefaultText:=BasicMessage("Tier"), DefaultTextColor:=White, DefaultShadowOffset:=ShadowOffset, DefaultShadowOpacity := DefaultTextShadowOpacity}
                                                                                                            overlay_slot: 
                                                                                                                Widget:=texture_block{DefaultDesiredSize:=vector2{X:=UPColumn2SizeX,Y:=UPColumn2SizeY},DefaultImage:=Images.EmptyTexture}
                                                                                                stack_box_slot: #Column 3
                                                                                                    Padding:=margin{Left:=InterColumnLeftPadding,Top:=InterColumnTopPadding,Right:=InterColumnRightPadding,Bottom:=InterColumnBottomPadding}
                                                                                                    Widget:=overlay:
                                                                                                        Slots:=array:
                                                                                                            overlay_slot: 
                                                                                                                Widget:=text_block{DefaultText:=BasicMessage("Level"), DefaultTextColor:=White, DefaultShadowOffset:=ShadowOffset, DefaultShadowOpacity := DefaultTextShadowOpacity}
                                                                                                            overlay_slot: 
                                                                                                                Widget:=texture_block{DefaultDesiredSize:=vector2{X:=UPColumn3SizeX,Y:=UPColumn3SizeY},DefaultImage:=Images.EmptyTexture}
                                                                                                stack_box_slot: #Column 4
                                                                                                    Padding:=margin{Left:=InterColumnLeftPadding,Top:=InterColumnTopPadding,Right:=InterColumnRightPadding,Bottom:=InterColumnBottomPadding}
                                                                                                    Widget:=overlay:
                                                                                                        Slots:=array:
                                                                                                            overlay_slot: 
                                                                                                                Widget:=text_block{DefaultText:=BasicMessage("Production"), DefaultTextColor:=White, DefaultShadowOffset:=ShadowOffset, DefaultShadowOpacity := DefaultTextShadowOpacity}
                                                                                                            overlay_slot: 
                                                                                                                Widget:=texture_block{DefaultDesiredSize:=vector2{X:=UPColumn4SizeX,Y:=UPColumn4SizeY},DefaultImage:=Images.EmptyTexture}
                                                                                                stack_box_slot: #Column 5
                                                                                                    Padding:=margin{Left:=InterColumnLeftPadding,Top:=InterColumnTopPadding,Right:=InterColumnRightPadding,Bottom:=InterColumnBottomPadding}
                                                                                                    Widget:=overlay:
                                                                                                        Slots:=array:
                                                                                                            overlay_slot: 
                                                                                                                Widget:=text_block{DefaultText:=BasicMessage("Upgrade"), DefaultTextColor:=White, DefaultShadowOffset:=ShadowOffset, DefaultShadowOpacity := DefaultTextShadowOpacity}
                                                                                                            overlay_slot: 
                                                                                                                Widget:=texture_block{DefaultDesiredSize:=vector2{X:=UPColumn5SizeX,Y:=UPColumn5SizeY},DefaultImage:=Images.EmptyTexture}
                                                                                                stack_box_slot: #Column 6
                                                                                                    Padding:=margin{Left:=InterColumnLeftPadding,Top:=InterColumnTopPadding,Right:=InterColumnRightPadding,Bottom:=InterColumnBottomPadding}
                                                                                                    Widget:=overlay:
                                                                                                        Slots:=array:
                                                                                                            overlay_slot: 
                                                                                                                Widget:=text_block{DefaultText:=BasicMessage("Upgrade cost"), DefaultTextColor:=White, DefaultShadowOffset:=ShadowOffset, DefaultShadowOpacity := DefaultTextShadowOpacity}
                                                                                                            overlay_slot: 
                                                                                                                Widget:=texture_block{DefaultDesiredSize:=vector2{X:=UPColumn6SizeX,Y:=UPColumn6SizeY},DefaultImage:=Images.EmptyTexture}
                                                                    stack_box_slot: #Line 2
                                                                        HorizontalAlignment:=horizontal_alignment.Center
                                                                        VerticalAlignment:=vertical_alignment.Center
                                                                        Padding:=margin{Left:=RowLeftPadding,Top:=RowTopPadding,Right:=RowRightPadding,Bottom:=RowBottomPadding}
                                                                        Widget:=overlay:
                                                                            Slots:=array:
                                                                                overlay_slot: #Line background
                                                                                    HorizontalAlignment:=horizontal_alignment.Center
                                                                                    VerticalAlignment:=vertical_alignment.Center
                                                                                    Widget:=texture_block{DefaultDesiredSize:=vector2{X:=LineSizeX,Y:=LineSizeY},DefaultImage:=Images.Img_LineBackground}
                                                                                overlay_slot: 
                                                                                    Widget:=stack_box:
                                                                                        Orientation:=orientation.Horizontal
                                                                                            Slots:=array:
                                                                                                stack_box_slot: #Column 1
                                                                                                    Padding:=margin{Left:=InterColumnLeftPadding,Top:=InterColumnTopPadding,Right:=InterColumnRightPadding,Bottom:=InterColumnBottomPadding}
                                                                                                    Widget:=overlay:
                                                                                                        Slots:=array:
                                                                                                            overlay_slot: 
                                                                                                                Widget:=UPRow1TextureBlock
                                                                                                stack_box_slot: #Column 2
                                                                                                    Padding:=margin{Left:=InterColumnLeftPadding,Top:=InterColumnTopPadding,Right:=InterColumnRightPadding,Bottom:=InterColumnBottomPadding}
                                                                                                    Widget:=overlay:
                                                                                                        Slots:=array:
                                                                                                            overlay_slot: 
                                                                                                                Widget:=UP_Row1_Tier_TextBlock
                                                                                                            overlay_slot: 
                                                                                                                Widget:=texture_block{DefaultDesiredSize:=vector2{X:=UPColumn2SizeX,Y:=UPColumn2SizeY},DefaultImage:=Images.EmptyTexture}
                                                                                                stack_box_slot: #Column 3
                                                                                                    Padding:=margin{Left:=InterColumnLeftPadding,Top:=InterColumnTopPadding,Right:=InterColumnRightPadding,Bottom:=InterColumnBottomPadding}
                                                                                                    Widget:=overlay:
                                                                                                        Slots:=array:
                                                                                                            overlay_slot: 
                                                                                                                Widget:=UP_Row1_Level_TextBlock
                                                                                                            overlay_slot: 
                                                                                                                Widget:=texture_block{DefaultDesiredSize:=vector2{X:=UPColumn3SizeX,Y:=UPColumn3SizeY},DefaultImage:=Images.EmptyTexture}
                                                                                                stack_box_slot: #Column 4
                                                                                                    Padding:=margin{Left:=InterColumnLeftPadding,Top:=InterColumnTopPadding,Right:=InterColumnRightPadding,Bottom:=InterColumnBottomPadding}
                                                                                                    Widget:=overlay:
                                                                                                        Slots:=array:
                                                                                                            overlay_slot:
                                                                                                                Widget:=UP_Row1_Production_TextBlock 
                                                                                                            overlay_slot: 
                                                                                                                Widget:=texture_block{DefaultDesiredSize:=vector2{X:=UPColumn4SizeX,Y:=UPColumn4SizeY},DefaultImage:=Images.EmptyTexture}
                                                                                                stack_box_slot: #Column 5
                                                                                                    Padding:=margin{Left:=InterColumnLeftPadding,Top:=InterColumnTopPadding,Right:=InterColumnRightPadding,Bottom:=InterColumnBottomPadding}
                                                                                                    Widget:=overlay:
                                                                                                        Slots:=array:
                                                                                                            overlay_slot: 
                                                                                                                Widget:=UP_Row1_Upgrade_TextBlock 
                                                                                                            overlay_slot: 
                                                                                                                Widget:=texture_block{DefaultDesiredSize:=vector2{X:=UPColumn5SizeX,Y:=UPColumn5SizeY},DefaultImage:=Images.EmptyTexture}
                                                                                                stack_box_slot: #Column 6
                                                                                                    Padding:=margin{Left:=InterColumnLeftPadding,Top:=InterColumnTopPadding,Right:=InterColumnRightPadding,Bottom:=InterColumnBottomPadding}
                                                                                                    Widget:=overlay:
                                                                                                        Slots:=array:
                                                                                                            overlay_slot: 
                                                                                                                Widget:=texture_block{DefaultDesiredSize:=vector2{X:=UPColumn6SizeX,Y:=UPColumn6SizeY},DefaultImage:=Images.EmptyTexture}
                                                                                                            overlay_slot: 
                                                                                                                HorizontalAlignment:=horizontal_alignment.Left
                                                                                                                VerticalAlignment:=vertical_alignment.Center 
                                                                                                                Padding:=margin{Left:=65.0,Top:=0.0,Right:=0.0,Bottom:=0.0}
                                                                                                                Widget:=UP_Row1_UpgradeCost_TextBlock 
                                                                                                            overlay_slot: 
                                                                                                                HorizontalAlignment:=horizontal_alignment.Left
                                                                                                                VerticalAlignment:=vertical_alignment.Center 
                                                                                                                Padding:=margin{Left:=10.0,Top:=0.0,Right:=0.0,Bottom:=0.0}
                                                                                                                Widget:=UP_Row1_SapphiresTexture
                                                                                                            overlay_slot:
                                                                                                                HorizontalAlignment:=horizontal_alignment.Fill
                                                                                                                VerticalAlignment:=vertical_alignment.Fill  
                                                                                                                Widget:=UP_Row1Button
                                                                    stack_box_slot: #Line 3
                                                                        HorizontalAlignment:=horizontal_alignment.Center
                                                                        VerticalAlignment:=vertical_alignment.Center
                                                                        Padding:=margin{Left:=RowLeftPadding,Top:=RowTopPadding,Right:=RowRightPadding,Bottom:=RowBottomPadding}
                                                                        Widget:=overlay:
                                                                            Slots:=array:
                                                                                overlay_slot: #Line background
                                                                                    HorizontalAlignment:=horizontal_alignment.Center
                                                                                    VerticalAlignment:=vertical_alignment.Center
                                                                                    Widget:=texture_block{DefaultDesiredSize:=vector2{X:=LineSizeX,Y:=LineSizeY},DefaultImage:=Images.Img_LineBackground}
                                                                                overlay_slot: 
                                                                                    Widget:=stack_box:
                                                                                        Orientation:=orientation.Horizontal
                                                                                            Slots:=array:
                                                                                                stack_box_slot: #Column 1
                                                                                                    Padding:=margin{Left:=InterColumnLeftPadding,Top:=InterColumnTopPadding,Right:=InterColumnRightPadding,Bottom:=InterColumnBottomPadding}
                                                                                                    Widget:=overlay:
                                                                                                        Slots:=array:
                                                                                                            overlay_slot: 
                                                                                                                Widget:=UPRow2TextureBlock
                                                                                                stack_box_slot: #Column 2
                                                                                                    Padding:=margin{Left:=InterColumnLeftPadding,Top:=InterColumnTopPadding,Right:=InterColumnRightPadding,Bottom:=InterColumnBottomPadding}
                                                                                                    Widget:=overlay:
                                                                                                        Slots:=array:
                                                                                                            overlay_slot: 
                                                                                                                Widget:=UP_Row2_Tier_TextBlock
                                                                                                            overlay_slot: 
                                                                                                                Widget:=texture_block{DefaultDesiredSize:=vector2{X:=UPColumn2SizeX,Y:=UPColumn2SizeY},DefaultImage:=Images.EmptyTexture}
                                                                                                stack_box_slot: #Column 3
                                                                                                    Padding:=margin{Left:=InterColumnLeftPadding,Top:=InterColumnTopPadding,Right:=InterColumnRightPadding,Bottom:=InterColumnBottomPadding}
                                                                                                    Widget:=overlay:
                                                                                                        Slots:=array:
                                                                                                            overlay_slot: 
                                                                                                                Widget:=UP_Row2_Level_TextBlock
                                                                                                            overlay_slot: 
                                                                                                                Widget:=texture_block{DefaultDesiredSize:=vector2{X:=UPColumn3SizeX,Y:=UPColumn3SizeY},DefaultImage:=Images.EmptyTexture}
                                                                                                stack_box_slot: #Column 4
                                                                                                    Padding:=margin{Left:=InterColumnLeftPadding,Top:=InterColumnTopPadding,Right:=InterColumnRightPadding,Bottom:=InterColumnBottomPadding}
                                                                                                    Widget:=overlay:
                                                                                                        Slots:=array:
                                                                                                            overlay_slot:
                                                                                                                Widget:=UP_Row2_Production_TextBlock
                                                                                                            overlay_slot: 
                                                                                                                Widget:=texture_block{DefaultDesiredSize:=vector2{X:=UPColumn4SizeX,Y:=UPColumn4SizeY},DefaultImage:=Images.EmptyTexture}
                                                                                                stack_box_slot: #Column 5
                                                                                                    Padding:=margin{Left:=InterColumnLeftPadding,Top:=InterColumnTopPadding,Right:=InterColumnRightPadding,Bottom:=InterColumnBottomPadding}
                                                                                                    Widget:=overlay:
                                                                                                        Slots:=array:
                                                                                                            overlay_slot: 
                                                                                                                Widget:=UP_Row2_Upgrade_TextBlock
                                                                                                            overlay_slot: 
                                                                                                                Widget:=texture_block{DefaultDesiredSize:=vector2{X:=UPColumn5SizeX,Y:=UPColumn5SizeY},DefaultImage:=Images.EmptyTexture}
                                                                                                stack_box_slot: #Column 6
                                                                                                    Padding:=margin{Left:=InterColumnLeftPadding,Top:=InterColumnTopPadding,Right:=InterColumnRightPadding,Bottom:=InterColumnBottomPadding}
                                                                                                    Widget:=overlay:
                                                                                                        Slots:=array:
                                                                                                            overlay_slot: 
                                                                                                                Widget:=texture_block{DefaultDesiredSize:=vector2{X:=UPColumn6SizeX,Y:=UPColumn6SizeY},DefaultImage:=Images.EmptyTexture}
                                                                                                            overlay_slot: 
                                                                                                                HorizontalAlignment:=horizontal_alignment.Left
                                                                                                                VerticalAlignment:=vertical_alignment.Center 
                                                                                                                Padding:=margin{Left:=65.0,Top:=0.0,Right:=0.0,Bottom:=0.0}
                                                                                                                Widget:=UP_Row2_UpgradeCost_TextBlock
                                                                                                            overlay_slot: 
                                                                                                                HorizontalAlignment:=horizontal_alignment.Left
                                                                                                                VerticalAlignment:=vertical_alignment.Center 
                                                                                                                Padding:=margin{Left:=10.0,Top:=0.0,Right:=0.0,Bottom:=0.0}
                                                                                                                Widget:=UP_Row2_SapphiresTexture
                                                                                                            overlay_slot:
                                                                                                                HorizontalAlignment:=horizontal_alignment.Fill
                                                                                                                VerticalAlignment:=vertical_alignment.Fill    
                                                                                                                Widget:=UP_Row2Button
                                                                    stack_box_slot: #Line 4
                                                                        HorizontalAlignment:=horizontal_alignment.Center
                                                                        VerticalAlignment:=vertical_alignment.Center
                                                                        Padding:=margin{Left:=RowLeftPadding,Top:=RowTopPadding,Right:=RowRightPadding,Bottom:=RowBottomPadding}
                                                                        Widget:=overlay:
                                                                            Slots:=array:
                                                                                overlay_slot: #Line background
                                                                                    HorizontalAlignment:=horizontal_alignment.Center
                                                                                    VerticalAlignment:=vertical_alignment.Center
                                                                                    Widget:=texture_block{DefaultDesiredSize:=vector2{X:=LineSizeX,Y:=LineSizeY},DefaultImage:=Images.Img_LineBackground}
                                                                                overlay_slot: 
                                                                                    Widget:=stack_box:
                                                                                        Orientation:=orientation.Horizontal
                                                                                            Slots:=array:
                                                                                                stack_box_slot: #Column 1
                                                                                                    Padding:=margin{Left:=InterColumnLeftPadding,Top:=InterColumnTopPadding,Right:=InterColumnRightPadding,Bottom:=InterColumnBottomPadding}
                                                                                                    Widget:=overlay:
                                                                                                        Slots:=array:
                                                                                                            overlay_slot: 
                                                                                                                Widget:=UPRow3TextureBlock
                                                                                                stack_box_slot: #Column 2
                                                                                                    Padding:=margin{Left:=InterColumnLeftPadding,Top:=InterColumnTopPadding,Right:=InterColumnRightPadding,Bottom:=InterColumnBottomPadding}
                                                                                                    Widget:=overlay:
                                                                                                        Slots:=array:
                                                                                                            overlay_slot: 
                                                                                                                Widget:=UP_Row3_Tier_TextBlock
                                                                                                            overlay_slot: 
                                                                                                                Widget:=texture_block{DefaultDesiredSize:=vector2{X:=UPColumn2SizeX,Y:=UPColumn2SizeY},DefaultImage:=Images.EmptyTexture}
                                                                                                stack_box_slot: #Column 3
                                                                                                    Padding:=margin{Left:=InterColumnLeftPadding,Top:=InterColumnTopPadding,Right:=InterColumnRightPadding,Bottom:=InterColumnBottomPadding}
                                                                                                    Widget:=overlay:
                                                                                                        Slots:=array:
                                                                                                            overlay_slot: 
                                                                                                                Widget:=UP_Row3_Level_TextBlock
                                                                                                            overlay_slot: 
                                                                                                                Widget:=texture_block{DefaultDesiredSize:=vector2{X:=UPColumn3SizeX,Y:=UPColumn3SizeY},DefaultImage:=Images.EmptyTexture}
                                                                                                stack_box_slot: #Column 4
                                                                                                    Padding:=margin{Left:=InterColumnLeftPadding,Top:=InterColumnTopPadding,Right:=InterColumnRightPadding,Bottom:=InterColumnBottomPadding}
                                                                                                    Widget:=overlay:
                                                                                                        Slots:=array:
                                                                                                            overlay_slot:
                                                                                                                Widget:=UP_Row3_Production_TextBlock
                                                                                                            overlay_slot: 
                                                                                                                Widget:=texture_block{DefaultDesiredSize:=vector2{X:=UPColumn4SizeX,Y:=UPColumn4SizeY},DefaultImage:=Images.EmptyTexture}
                                                                                                stack_box_slot: #Column 5
                                                                                                    Padding:=margin{Left:=InterColumnLeftPadding,Top:=InterColumnTopPadding,Right:=InterColumnRightPadding,Bottom:=InterColumnBottomPadding}
                                                                                                    Widget:=overlay:
                                                                                                        Slots:=array:
                                                                                                            overlay_slot: 
                                                                                                                Widget:=UP_Row3_Upgrade_TextBlock
                                                                                                            overlay_slot: 
                                                                                                                Widget:=texture_block{DefaultDesiredSize:=vector2{X:=UPColumn5SizeX,Y:=UPColumn5SizeY},DefaultImage:=Images.EmptyTexture}
                                                                                                stack_box_slot: #Column 6
                                                                                                    Padding:=margin{Left:=InterColumnLeftPadding,Top:=InterColumnTopPadding,Right:=InterColumnRightPadding,Bottom:=InterColumnBottomPadding}
                                                                                                    Widget:=overlay:
                                                                                                        Slots:=array:
                                                                                                            overlay_slot: 
                                                                                                                Widget:=texture_block{DefaultDesiredSize:=vector2{X:=UPColumn6SizeX,Y:=UPColumn6SizeY},DefaultImage:=Images.EmptyTexture}
                                                                                                            overlay_slot: 
                                                                                                                HorizontalAlignment:=horizontal_alignment.Left
                                                                                                                VerticalAlignment:=vertical_alignment.Center 
                                                                                                                Padding:=margin{Left:=65.0,Top:=0.0,Right:=0.0,Bottom:=0.0}
                                                                                                                Widget:=UP_Row3_UpgradeCost_TextBlock
                                                                                                            overlay_slot: 
                                                                                                                HorizontalAlignment:=horizontal_alignment.Left
                                                                                                                VerticalAlignment:=vertical_alignment.Center 
                                                                                                                Padding:=margin{Left:=10.0,Top:=0.0,Right:=0.0,Bottom:=0.0}
                                                                                                                Widget:=UP_Row3_SapphiresTexture
                                                                                                            overlay_slot:
                                                                                                                HorizontalAlignment:=horizontal_alignment.Fill
                                                                                                                VerticalAlignment:=vertical_alignment.Fill    
                                                                                                                Widget:=UP_Row3Button
        
                                                                    stack_box_slot: #Line 5
                                                                        HorizontalAlignment:=horizontal_alignment.Center
                                                                        VerticalAlignment:=vertical_alignment.Center
                                                                        Padding:=margin{Left:=RowLeftPadding,Top:=RowTopPadding,Right:=RowRightPadding,Bottom:=RowBottomPadding}
                                                                        Widget:=overlay:
                                                                            Slots:=array:
                                                                                overlay_slot: #Line background
                                                                                    HorizontalAlignment:=horizontal_alignment.Center
                                                                                    VerticalAlignment:=vertical_alignment.Center
                                                                                    Widget:=texture_block{DefaultDesiredSize:=vector2{X:=LineSizeX,Y:=LineSizeY},DefaultImage:=Images.Img_LineBackground}
                                                                                overlay_slot: 
                                                                                    Widget:=stack_box:
                                                                                        Orientation:=orientation.Horizontal
                                                                                            Slots:=array:
                                                                                                stack_box_slot: #Column 1
                                                                                                    Padding:=margin{Left:=InterColumnLeftPadding,Top:=InterColumnTopPadding,Right:=InterColumnRightPadding,Bottom:=InterColumnBottomPadding}
                                                                                                    Widget:=overlay:
                                                                                                        Slots:=array:
                                                                                                            overlay_slot: 
                                                                                                                Widget:=UPRow4TextureBlock
                                                                                                stack_box_slot: #Column 2
                                                                                                    Padding:=margin{Left:=InterColumnLeftPadding,Top:=InterColumnTopPadding,Right:=InterColumnRightPadding,Bottom:=InterColumnBottomPadding}
                                                                                                    Widget:=overlay:
                                                                                                        Slots:=array:
                                                                                                            overlay_slot: 
                                                                                                                Widget:=UP_Row4_Tier_TextBlock
                                                                                                            overlay_slot: 
                                                                                                                Widget:=texture_block{DefaultDesiredSize:=vector2{X:=UPColumn2SizeX,Y:=UPColumn2SizeY},DefaultImage:=Images.EmptyTexture}
                                                                                                stack_box_slot: #Column 3
                                                                                                    Padding:=margin{Left:=InterColumnLeftPadding,Top:=InterColumnTopPadding,Right:=InterColumnRightPadding,Bottom:=InterColumnBottomPadding}
                                                                                                    Widget:=overlay:
                                                                                                        Slots:=array:
                                                                                                            overlay_slot: 
                                                                                                                Widget:=UP_Row4_Level_TextBlock
                                                                                                            overlay_slot: 
                                                                                                                Widget:=texture_block{DefaultDesiredSize:=vector2{X:=UPColumn3SizeX,Y:=UPColumn3SizeY},DefaultImage:=Images.EmptyTexture}
                                                                                                stack_box_slot: #Column 4
                                                                                                    Padding:=margin{Left:=InterColumnLeftPadding,Top:=InterColumnTopPadding,Right:=InterColumnRightPadding,Bottom:=InterColumnBottomPadding}
                                                                                                    Widget:=overlay:
                                                                                                        Slots:=array:
                                                                                                            overlay_slot:
                                                                                                                Widget:=UP_Row4_Production_TextBlock
                                                                                                            overlay_slot: 
                                                                                                                Widget:=texture_block{DefaultDesiredSize:=vector2{X:=UPColumn4SizeX,Y:=UPColumn4SizeY},DefaultImage:=Images.EmptyTexture}
                                                                                                stack_box_slot: #Column 5
                                                                                                    Padding:=margin{Left:=InterColumnLeftPadding,Top:=InterColumnTopPadding,Right:=InterColumnRightPadding,Bottom:=InterColumnBottomPadding}
                                                                                                    Widget:=overlay:
                                                                                                        Slots:=array:
                                                                                                            overlay_slot: 
                                                                                                                Widget:=UP_Row4_Upgrade_TextBlock
                                                                                                            overlay_slot: 
                                                                                                                Widget:=texture_block{DefaultDesiredSize:=vector2{X:=UPColumn5SizeX,Y:=UPColumn5SizeY},DefaultImage:=Images.EmptyTexture}
                                                                                                stack_box_slot: #Column 6
                                                                                                    Padding:=margin{Left:=InterColumnLeftPadding,Top:=InterColumnTopPadding,Right:=InterColumnRightPadding,Bottom:=InterColumnBottomPadding}
                                                                                                    Widget:=overlay:
                                                                                                        Slots:=array:
                                                                                                            overlay_slot: 
                                                                                                                Widget:=texture_block{DefaultDesiredSize:=vector2{X:=UPColumn6SizeX,Y:=UPColumn6SizeY},DefaultImage:=Images.EmptyTexture}
                                                                                                            overlay_slot: 
                                                                                                                HorizontalAlignment:=horizontal_alignment.Left
                                                                                                                VerticalAlignment:=vertical_alignment.Center 
                                                                                                                Padding:=margin{Left:=65.0,Top:=0.0,Right:=0.0,Bottom:=0.0}
                                                                                                                Widget:=UP_Row4_UpgradeCost_TextBlock
                                                                                                            overlay_slot: 
                                                                                                                HorizontalAlignment:=horizontal_alignment.Left
                                                                                                                VerticalAlignment:=vertical_alignment.Center 
                                                                                                                Padding:=margin{Left:=10.0,Top:=0.0,Right:=0.0,Bottom:=0.0}
                                                                                                                Widget:=UP_Row4_SapphiresTexture
                                                                                                            overlay_slot:
                                                                                                                HorizontalAlignment:=horizontal_alignment.Fill
                                                                                                                VerticalAlignment:=vertical_alignment.Fill    
                                                                                                                Widget:=UP_Row4Button
        
                                                                    stack_box_slot: #Line 6
                                                                        HorizontalAlignment:=horizontal_alignment.Center
                                                                        VerticalAlignment:=vertical_alignment.Center
                                                                        Padding:=margin{Left:=RowLeftPadding,Top:=RowTopPadding,Right:=RowRightPadding,Bottom:=RowBottomPadding}
                                                                        Widget:=overlay:
                                                                            Slots:=array:
                                                                                overlay_slot: #Line background
                                                                                    HorizontalAlignment:=horizontal_alignment.Center
                                                                                    VerticalAlignment:=vertical_alignment.Center
                                                                                    Widget:=texture_block{DefaultDesiredSize:=vector2{X:=LineSizeX,Y:=LineSizeY},DefaultImage:=Images.Img_LineBackground}
                                                                                overlay_slot: 
                                                                                    Widget:=stack_box:
                                                                                        Orientation:=orientation.Horizontal
                                                                                            Slots:=array:
                                                                                                stack_box_slot: #Column 1
                                                                                                    Padding:=margin{Left:=InterColumnLeftPadding,Top:=InterColumnTopPadding,Right:=InterColumnRightPadding,Bottom:=InterColumnBottomPadding}
                                                                                                    Widget:=overlay:
                                                                                                        Slots:=array:
                                                                                                            overlay_slot: 
                                                                                                                Widget:=UPRow5TextureBlock
                                                                                                stack_box_slot: #Column 2
                                                                                                    Padding:=margin{Left:=InterColumnLeftPadding,Top:=InterColumnTopPadding,Right:=InterColumnRightPadding,Bottom:=InterColumnBottomPadding}
                                                                                                    Widget:=overlay:
                                                                                                        Slots:=array:
                                                                                                            overlay_slot: 
                                                                                                                Widget:=UP_Row5_Tier_TextBlock
                                                                                                            overlay_slot: 
                                                                                                                Widget:=texture_block{DefaultDesiredSize:=vector2{X:=UPColumn2SizeX,Y:=UPColumn2SizeY},DefaultImage:=Images.EmptyTexture}
                                                                                                stack_box_slot: #Column 3
                                                                                                    Padding:=margin{Left:=InterColumnLeftPadding,Top:=InterColumnTopPadding,Right:=InterColumnRightPadding,Bottom:=InterColumnBottomPadding}
                                                                                                    Widget:=overlay:
                                                                                                        Slots:=array:
                                                                                                            overlay_slot: 
                                                                                                                Widget:=UP_Row5_Level_TextBlock
                                                                                                            overlay_slot: 
                                                                                                                Widget:=texture_block{DefaultDesiredSize:=vector2{X:=UPColumn3SizeX,Y:=UPColumn3SizeY},DefaultImage:=Images.EmptyTexture}
                                                                                                stack_box_slot: #Column 4
                                                                                                    Padding:=margin{Left:=InterColumnLeftPadding,Top:=InterColumnTopPadding,Right:=InterColumnRightPadding,Bottom:=InterColumnBottomPadding}
                                                                                                    Widget:=overlay:
                                                                                                        Slots:=array:
                                                                                                            overlay_slot:
                                                                                                                Widget:=UP_Row5_Production_TextBlock
                                                                                                            overlay_slot: 
                                                                                                                Widget:=texture_block{DefaultDesiredSize:=vector2{X:=UPColumn4SizeX,Y:=UPColumn4SizeY},DefaultImage:=Images.EmptyTexture}
                                                                                                stack_box_slot: #Column 5
                                                                                                    Padding:=margin{Left:=InterColumnLeftPadding,Top:=InterColumnTopPadding,Right:=InterColumnRightPadding,Bottom:=InterColumnBottomPadding}
                                                                                                    Widget:=overlay:
                                                                                                        Slots:=array:
                                                                                                            overlay_slot: 
                                                                                                                Widget:=UP_Row5_Upgrade_TextBlock
                                                                                                            overlay_slot: 
                                                                                                                Widget:=texture_block{DefaultDesiredSize:=vector2{X:=UPColumn5SizeX,Y:=UPColumn5SizeY},DefaultImage:=Images.EmptyTexture}
                                                                                                stack_box_slot: #Column 6
                                                                                                    Padding:=margin{Left:=InterColumnLeftPadding,Top:=InterColumnTopPadding,Right:=InterColumnRightPadding,Bottom:=InterColumnBottomPadding}
                                                                                                    Widget:=overlay:
                                                                                                        Slots:=array:
                                                                                                            overlay_slot: 
                                                                                                                Widget:=texture_block{DefaultDesiredSize:=vector2{X:=UPColumn6SizeX,Y:=UPColumn6SizeY},DefaultImage:=Images.EmptyTexture}
                                                                                                            overlay_slot: 
                                                                                                                HorizontalAlignment:=horizontal_alignment.Left
                                                                                                                VerticalAlignment:=vertical_alignment.Center 
                                                                                                                Padding:=margin{Left:=65.0,Top:=0.0,Right:=0.0,Bottom:=0.0}
                                                                                                                Widget:=UP_Row5_UpgradeCost_TextBlock
                                                                                                            overlay_slot: 
                                                                                                                HorizontalAlignment:=horizontal_alignment.Left
                                                                                                                VerticalAlignment:=vertical_alignment.Center 
                                                                                                                Padding:=margin{Left:=10.0,Top:=0.0,Right:=0.0,Bottom:=0.0}
                                                                                                                Widget:=UP_Row5_SapphiresTexture
                                                                                                            overlay_slot:
                                                                                                                HorizontalAlignment:=horizontal_alignment.Fill
                                                                                                                VerticalAlignment:=vertical_alignment.Fill    
                                                                                                                Widget:=UP_Row5Button
        
                                                                    stack_box_slot: #Line 7
                                                                        HorizontalAlignment:=horizontal_alignment.Fill
                                                                        VerticalAlignment:=vertical_alignment.Fill
                                                                        Padding:=margin{Left:=RowLeftPadding,Top:=RowTopPadding,Right:=RowRightPadding,Bottom:=RowBottomPadding}
                                                                        Widget:=overlay:
                                                                            Slots:=array:
                                                                                overlay_slot: 
                                                                                    Widget:=stack_box:
                                                                                        Orientation:=orientation.Horizontal
                                                                                            Slots:=array:
                                                                                                stack_box_slot: #Column 1
                                                                                                    Padding:=margin{Left:=InterColumnLeftPadding,Top:=InterColumnTopPadding,Right:=InterColumnRightPadding,Bottom:=InterColumnBottomPadding}
                                                                                                    Widget:=overlay:
                                                                                                        Slots:=array:
                                                                                                            overlay_slot: 
                                                                                                                Widget:=texture_block{DefaultDesiredSize:=vector2{X:=800.0,Y:=80.0},DefaultImage:=Images.EmptyTexture}
                                                                                                            overlay_slot: 
                                                                                                                Widget:=UP_NumPage_TextBlock
                                                                    stack_box_slot: #Line 8
                                                                        HorizontalAlignment:=horizontal_alignment.Fill
                                                                        VerticalAlignment:=vertical_alignment.Fill
                                                                        Padding:=margin{Left:=RowLeftPadding,Top:=RowTopPadding,Right:=RowRightPadding,Bottom:=RowBottomPadding}
                                                                        Widget:=overlay:
                                                                            Slots:=array:
                                                                                overlay_slot: 
                                                                                    Widget:=stack_box:
                                                                                        Orientation:=orientation.Horizontal
                                                                                            Slots:=array:
                                                                                                stack_box_slot: #Column 1
                                                                                                    Padding:=margin{Left:=InterColumnLeftPadding,Top:=InterColumnTopPadding,Right:=InterColumnRightPadding,Bottom:=InterColumnBottomPadding}
                                                                                                    Widget:=overlay:
                                                                                                        Slots:=array:
                                                                                                            overlay_slot: 
                                                                                                                Widget:=texture_block{DefaultDesiredSize:=vector2{X:=1.0,Y:=BigCloseButtonSizeY},DefaultImage:=Images.EmptyTexture} 
                                                                                                            overlay_slot: 
                                                                                                                VerticalAlignment:=vertical_alignment.Fill
                                                                                                                Widget:=UP_PrevPageButton
                                                                                                stack_box_slot: #Column 2
                                                                                                    Padding:=margin{Left:=InterColumnLeftPadding,Top:=InterColumnTopPadding,Right:=InterColumnRightPadding,Bottom:=InterColumnBottomPadding}
                                                                                                    Widget:=overlay:
                                                                                                        Slots:=array: 
                                                                                                            overlay_slot: 
                                                                                                                Widget:=texture_block{DefaultDesiredSize:=vector2{X:=1.0,Y:=BigCloseButtonSizeY},DefaultImage:=Images.EmptyTexture} 
                                                                                                            overlay_slot: 
                                                                                                                VerticalAlignment:=vertical_alignment.Fill
                                                                                                                Widget:=UP_NextPageButton
                                                                                                
                                                    stack_box_slot:
                                                        HorizontalAlignment:=horizontal_alignment.Fill
                                                        VerticalAlignment:=vertical_alignment.Fill
                                                        Widget:= texture_block{DefaultImage:=Images.Img_VerticalLine}
                                                    stack_box_slot:
                                                        HorizontalAlignment:=horizontal_alignment.Fill
                                                        VerticalAlignment:=vertical_alignment.Fill
                                                        #Baker speed stack box
                                                        Widget:=stack_box:
                                                            Orientation:=orientation.Vertical
                                                                Slots:=array: 
                                                                    stack_box_slot: #Line 0
                                                                        HorizontalAlignment:=horizontal_alignment.Right
                                                                        VerticalAlignment:=vertical_alignment.Top
                                                                        Padding:=margin{Left:=RowLeftPadding,Top:=RowTopPadding,Right:=RowRightPadding,Bottom:=RowBottomPadding}
                                                                        Widget:=overlay:
                                                                            Slots:=array: 
                                                                                overlay_slot:
                                                                                    HorizontalAlignment:=horizontal_alignment.Right
                                                                                    VerticalAlignment:=vertical_alignment.Center
                                                                                    Widget:=overlay:
                                                                                        Slots:=array: 
                                                                                            overlay_slot:
                                                                                                Widget:=texture_block{DefaultDesiredSize:=vector2{X:=SizeBoxSaphiresX,Y:=SizeBoxSaphiresY},DefaultImage:=Images.EmptyTexture} 
                                                                                            overlay_slot:
                                                                                                HorizontalAlignment:=horizontal_alignment.Fill
                                                                                                VerticalAlignment:=vertical_alignment.Fill
                                                                                                Widget:=texture_block{DefaultImage:=Images.Img_LineBackground2}  
                                                                                            overlay_slot:
                                                                                                HorizontalAlignment:=horizontal_alignment.Left
                                                                                                VerticalAlignment:=vertical_alignment.Center
                                                                                                Padding:=margin{Left:=Icon_LeftMargin}
                                                                                                Widget:=texture_block{DefaultDesiredSize:=vector2{X:=IconSmallSizeX,Y:=IconSmallSizeY},DefaultImage:=Images.Sapphire_Icon}
                                                                                            overlay_slot:
                                                                                                HorizontalAlignment:=horizontal_alignment.Left
                                                                                                VerticalAlignment:=vertical_alignment.Center
                                                                                                Padding:=margin{Left:=TextIcon_LeftMargin}
                                                                                                Widget:=TotalSapphires_TextBlock  
                                                                    stack_box_slot: #Line 1
                                                                        Padding:=margin{Left:=RowLeftPadding,Top:=RowTopPadding,Right:=RowRightPadding,Bottom:=RowBottomPadding}
                                                                        HorizontalAlignment:=horizontal_alignment.Fill
                                                                        VerticalAlignment:=vertical_alignment.Bottom
                                                                        Widget:=overlay:
                                                                            Slots:=array:
                                                                                overlay_slot: 
                                                                                    HorizontalAlignment:=horizontal_alignment.Fill
                                                                                    VerticalAlignment:=vertical_alignment.Bottom
                                                                                    Widget:=texture_block{DefaultDesiredSize:=vector2{X:=HorizontalLineSizeX,Y:=HorizontalLineSizeY},DefaultImage:=Images.Img_HorizontalLine}
                                                                                overlay_slot: 
                                                                                    HorizontalAlignment:=horizontal_alignment.Center
                                                                                    VerticalAlignment:=vertical_alignment.Center 
                                                                                    Widget:=text_block{DefaultText:=BasicMessage("Baker speed & efficiency"), DefaultTextColor:=White, DefaultShadowOffset:=ShadowOffset, DefaultShadowOpacity := DefaultTextShadowOpacity}
                                                                                overlay_slot: 
                                                                                    HorizontalAlignment:=horizontal_alignment.Center
                                                                                    VerticalAlignment:=vertical_alignment.Center 
                                                                                    Widget:=texture_block{DefaultDesiredSize:=vector2{X:=10.0,Y:=LineSizeY},DefaultImage:=Images.EmptyTexture}
                                                                    stack_box_slot: #Line 2
                                                                        Padding:=margin{Left:=RowLeftPadding,Top:=RowTopPadding,Right:=RowRightPadding,Bottom:=RowBottomPadding}
                                                                            Widget:=overlay:
                                                                                Slots:=array:
                                                                                    overlay_slot: 
                                                                                        HorizontalAlignment:=horizontal_alignment.Fill
                                                                                        VerticalAlignment:=vertical_alignment.Center
                                                                                        Widget:=texture_block{DefaultDesiredSize:=vector2{X:=0.0,Y:=LineSizeY},DefaultImage:=Images.Img_LineBackground}
                                                                                    overlay_slot: 
                                                                                        Widget:=stack_box:
                                                                                            Orientation:=orientation.Horizontal
                                                                                                Slots:=array:
                                                                                                    stack_box_slot: 
                                                                                                        Padding:=margin{Left:=InterColumnLeftPadding,Top:=InterColumnTopPadding,Right:=InterColumnRightPadding,Bottom:=InterColumnBottomPadding}
                                                                                                        Widget:=overlay:
                                                                                                            Slots:=array:
                                                                                                                overlay_slot:
                                                                                                                    Widget:=text_block{DefaultText:=BasicMessage("Perk"), DefaultTextColor:=White, DefaultShadowOffset:=ShadowOffset, DefaultShadowOpacity := DefaultTextShadowOpacity}
                                                                                                                overlay_slot: 
                                                                                                                    Widget:=texture_block{DefaultDesiredSize:=vector2{X:=SREColumn1SizeX,Y:=SREColumn1SizeY},DefaultImage:=Images.EmptyTexture}
                                                                                                    stack_box_slot: 
                                                                                                        Padding:=margin{Left:=InterColumnLeftPadding,Top:=InterColumnTopPadding,Right:=InterColumnRightPadding,Bottom:=InterColumnBottomPadding}
                                                                                                        Widget:=overlay:
                                                                                                            Slots:=array:
                                                                                                                overlay_slot:
                                                                                                                    Widget:=text_block{DefaultText:=BasicMessage("Current"), DefaultTextColor:=White, DefaultShadowOffset:=ShadowOffset, DefaultShadowOpacity := DefaultTextShadowOpacity}
                                                                                                                overlay_slot: 
                                                                                                                    Widget:=texture_block{DefaultDesiredSize:=vector2{X:=SREColumn2SizeX,Y:=SREColumn2SizeY},DefaultImage:=Images.EmptyTexture}
                                                                                                    stack_box_slot: 
                                                                                                        Padding:=margin{Left:=InterColumnLeftPadding,Top:=InterColumnTopPadding,Right:=InterColumnRightPadding,Bottom:=InterColumnBottomPadding}
                                                                                                        Widget:=overlay:
                                                                                                            Slots:=array:
                                                                                                                overlay_slot:
                                                                                                                    Widget:=text_block{DefaultText:=BasicMessage("Upgrade"), DefaultTextColor:=White, DefaultShadowOffset:=ShadowOffset, DefaultShadowOpacity := DefaultTextShadowOpacity}
                                                                                                                overlay_slot: 
                                                                                                                    Widget:=texture_block{DefaultDesiredSize:=vector2{X:=SREColumn3SizeX,Y:=SREColumn3SizeY},DefaultImage:=Images.EmptyTexture}
                                                                                                    stack_box_slot: 
                                                                                                        Padding:=margin{Left:=InterColumnLeftPadding,Top:=InterColumnTopPadding,Right:=InterColumnRightPadding,Bottom:=InterColumnBottomPadding}
                                                                                                        Widget:=overlay:
                                                                                                            Slots:=array:
                                                                                                                overlay_slot:
                                                                                                                    Widget:=text_block{DefaultText:=BasicMessage("Upgrade cost"), DefaultTextColor:=White, DefaultShadowOffset:=ShadowOffset, DefaultShadowOpacity := DefaultTextShadowOpacity}
                                                                                                                overlay_slot: 
                                                                                                                    Widget:=texture_block{DefaultDesiredSize:=vector2{X:=SREColumn4SizeX,Y:=SREColumn4SizeY},DefaultImage:=Images.EmptyTexture}
                                                                    stack_box_slot: #Line 3
                                                                        Padding:=margin{Left:=RowLeftPadding,Top:=RowTopPadding,Right:=RowRightPadding,Bottom:=RowBottomPadding}
                                                                            Widget:=overlay:
                                                                                Slots:=array:
                                                                                    overlay_slot: 
                                                                                        HorizontalAlignment:=horizontal_alignment.Fill
                                                                                        VerticalAlignment:=vertical_alignment.Center
                                                                                        Widget:=texture_block{DefaultDesiredSize:=vector2{X:=0.0,Y:=LineSizeY},DefaultImage:=Images.Img_LineBackground}
                                                                                    overlay_slot: 
                                                                                        Widget:=stack_box:
                                                                                            Orientation:=orientation.Horizontal
                                                                                                Slots:=array:
                                                                                                    stack_box_slot: 
                                                                                                        Padding:=margin{Left:=InterColumnLeftPadding,Top:=InterColumnTopPadding,Right:=InterColumnRightPadding,Bottom:=InterColumnBottomPadding}
                                                                                                        Widget:=overlay:
                                                                                                            Slots:=array:
                                                                                                                overlay_slot:
                                                                                                                    Widget:=BS_Row_Level_IconTexture
                                                                                                                overlay_slot: 
                                                                                                                    Widget:=texture_block{DefaultDesiredSize:=vector2{X:=SREColumn1SizeX,Y:=SREColumn1SizeY},DefaultImage:=Images.EmptyTexture}
                                                                                                    stack_box_slot: 
                                                                                                        Padding:=margin{Left:=InterColumnLeftPadding,Top:=InterColumnTopPadding,Right:=InterColumnRightPadding,Bottom:=InterColumnBottomPadding}
                                                                                                        Widget:=overlay:
                                                                                                            Slots:=array:
                                                                                                                overlay_slot:
                                                                                                                    Widget:=BS_Row_Speed_TextBlock
                                                                                                                overlay_slot: 
                                                                                                                    Widget:=texture_block{DefaultDesiredSize:=vector2{X:=SREColumn2SizeX,Y:=SREColumn2SizeY},DefaultImage:=Images.EmptyTexture}
                                                                                                    stack_box_slot: 
                                                                                                        Padding:=margin{Left:=InterColumnLeftPadding,Top:=InterColumnTopPadding,Right:=InterColumnRightPadding,Bottom:=InterColumnBottomPadding}
                                                                                                        Widget:=overlay:
                                                                                                            Slots:=array:
                                                                                                                overlay_slot:
                                                                                                                    Widget:=BS_Row_Upgrade_TextBlock
                                                                                                                overlay_slot: 
                                                                                                                    Widget:=texture_block{DefaultDesiredSize:=vector2{X:=SREColumn3SizeX,Y:=SREColumn3SizeY},DefaultImage:=Images.EmptyTexture}
                                                                                                    stack_box_slot: 
                                                                                                        Padding:=margin{Left:=InterColumnLeftPadding,Top:=InterColumnTopPadding,Right:=InterColumnRightPadding,Bottom:=InterColumnBottomPadding}
                                                                                                        Widget:=overlay:
                                                                                                            Slots:=array:
                                                                                                                overlay_slot: 
                                                                                                                    Widget:=texture_block{DefaultDesiredSize:=vector2{X:=SREColumn4SizeX,Y:=SREColumn4SizeY},DefaultImage:=Images.EmptyTexture} 
                                                                                                                overlay_slot: 
                                                                                                                    HorizontalAlignment:=horizontal_alignment.Left
                                                                                                                    VerticalAlignment:=vertical_alignment.Center 
                                                                                                                    Padding:=margin{Left:=65.0,Top:=0.0,Right:=0.0,Bottom:=0.0}
                                                                                                                    Widget:=BS_Row_UpgradeCost_TextBlock               
                                                                                                                overlay_slot: 
                                                                                                                    HorizontalAlignment:=horizontal_alignment.Left
                                                                                                                    VerticalAlignment:=vertical_alignment.Center 
                                                                                                                    Padding:=margin{Left:=10.0,Top:=0.0,Right:=0.0,Bottom:=0.0}
                                                                                                                    Widget:=BS_Row_SapphiresTexture
                                                                                                                overlay_slot:
                                                                                                                    HorizontalAlignment:=horizontal_alignment.Fill
                                                                                                                    VerticalAlignment:=vertical_alignment.Fill  
                                                                                                                    Widget:=BS_RowButton
                                                                    stack_box_slot: #Line 4
                                                                        Padding:=margin{Left:=RowLeftPadding,Top:=RowTopPadding,Right:=RowRightPadding,Bottom:=RowBottomPadding}
                                                                            Widget:=overlay:
                                                                                Slots:=array:
                                                                                    overlay_slot: 
                                                                                        HorizontalAlignment:=horizontal_alignment.Fill
                                                                                        VerticalAlignment:=vertical_alignment.Center
                                                                                        Widget:=RE_Row1_LineTexture
                                                                                    overlay_slot: 
                                                                                        Widget:=stack_box:
                                                                                            Orientation:=orientation.Horizontal
                                                                                                Slots:=array:
                                                                                                    stack_box_slot: 
                                                                                                        Padding:=margin{Left:=InterColumnLeftPadding,Top:=InterColumnTopPadding,Right:=InterColumnRightPadding,Bottom:=InterColumnBottomPadding}
                                                                                                        Widget:=overlay:
                                                                                                            Slots:=array:
                                                                                                                overlay_slot:
                                                                                                                    Widget:=RE_Row1_IconTexture
                                                                                                                overlay_slot: 
                                                                                                                    Widget:=texture_block{DefaultDesiredSize:=vector2{X:=SREColumn1SizeX,Y:=SREColumn1SizeY},DefaultImage:=Images.EmptyTexture}
                                                                                                    stack_box_slot: 
                                                                                                        Padding:=margin{Left:=InterColumnLeftPadding,Top:=InterColumnTopPadding,Right:=InterColumnRightPadding,Bottom:=InterColumnBottomPadding}
                                                                                                        Widget:=overlay:
                                                                                                            Slots:=array:
                                                                                                                overlay_slot:
                                                                                                                    Widget:=RE_Row1_Efficiency_TextBlock
                                                                                                                overlay_slot: 
                                                                                                                    Widget:=texture_block{DefaultDesiredSize:=vector2{X:=SREColumn2SizeX,Y:=SREColumn2SizeY},DefaultImage:=Images.EmptyTexture}
                                                                                                    stack_box_slot: 
                                                                                                        Padding:=margin{Left:=InterColumnLeftPadding,Top:=InterColumnTopPadding,Right:=InterColumnRightPadding,Bottom:=InterColumnBottomPadding}
                                                                                                        Widget:=overlay:
                                                                                                            Slots:=array:
                                                                                                                overlay_slot:
                                                                                                                    Widget:=RE_Row1_Upgrade_TextBlock
                                                                                                                overlay_slot: 
                                                                                                                    Widget:=texture_block{DefaultDesiredSize:=vector2{X:=SREColumn3SizeX,Y:=SREColumn3SizeY},DefaultImage:=Images.EmptyTexture}
                                                                                                    stack_box_slot: 
                                                                                                        Padding:=margin{Left:=InterColumnLeftPadding,Top:=InterColumnTopPadding,Right:=InterColumnRightPadding,Bottom:=InterColumnBottomPadding}
                                                                                                        Widget:=overlay:
                                                                                                            Slots:=array:   
                                                                                                                overlay_slot: 
                                                                                                                    Widget:=texture_block{DefaultDesiredSize:=vector2{X:=SREColumn4SizeX,Y:=SREColumn4SizeY},DefaultImage:=Images.EmptyTexture} 
                                                                                                                overlay_slot: 
                                                                                                                    HorizontalAlignment:=horizontal_alignment.Left
                                                                                                                    VerticalAlignment:=vertical_alignment.Center 
                                                                                                                    Padding:=margin{Left:=65.0,Top:=0.0,Right:=0.0,Bottom:=0.0}
                                                                                                                    Widget:=RE_Row1_UpgradeCost_TextBlock                
                                                                                                                overlay_slot: 
                                                                                                                    HorizontalAlignment:=horizontal_alignment.Left
                                                                                                                    VerticalAlignment:=vertical_alignment.Center 
                                                                                                                    Padding:=margin{Left:=10.0,Top:=0.0,Right:=0.0,Bottom:=0.0}
                                                                                                                    Widget:=RE_Row1_SapphiresTexture
                                                                                                                overlay_slot:
                                                                                                                    HorizontalAlignment:=horizontal_alignment.Fill
                                                                                                                    VerticalAlignment:=vertical_alignment.Fill  
                                                                                                                    Widget:=RE_Row1Button
                                                                    stack_box_slot: #Line 5
                                                                        Padding:=margin{Left:=RowLeftPadding,Top:=RowTopPadding,Right:=RowRightPadding,Bottom:=RowBottomPadding}
                                                                            Widget:=overlay:
                                                                                Slots:=array:
                                                                                    overlay_slot: 
                                                                                        HorizontalAlignment:=horizontal_alignment.Fill
                                                                                        VerticalAlignment:=vertical_alignment.Center
                                                                                        Widget:=RE_Row2_LineTexture
                                                                                    overlay_slot: 
                                                                                        Widget:=stack_box:
                                                                                            Orientation:=orientation.Horizontal
                                                                                                Slots:=array:
                                                                                                    stack_box_slot: 
                                                                                                        Padding:=margin{Left:=InterColumnLeftPadding,Top:=InterColumnTopPadding,Right:=InterColumnRightPadding,Bottom:=InterColumnBottomPadding}
                                                                                                        Widget:=overlay:
                                                                                                            Slots:=array:
                                                                                                                overlay_slot:
                                                                                                                    Widget:=RE_Row2_IconTexture
                                                                                                                overlay_slot: 
                                                                                                                    Widget:=texture_block{DefaultDesiredSize:=vector2{X:=SREColumn1SizeX,Y:=SREColumn1SizeY},DefaultImage:=Images.EmptyTexture}
                                                                                                    stack_box_slot: 
                                                                                                        Padding:=margin{Left:=InterColumnLeftPadding,Top:=InterColumnTopPadding,Right:=InterColumnRightPadding,Bottom:=InterColumnBottomPadding}
                                                                                                        Widget:=overlay:
                                                                                                            Slots:=array:
                                                                                                                overlay_slot:
                                                                                                                    Widget:=RE_Row2_Efficiency_TextBlock
                                                                                                                overlay_slot: 
                                                                                                                    Widget:=texture_block{DefaultDesiredSize:=vector2{X:=SREColumn2SizeX,Y:=SREColumn2SizeY},DefaultImage:=Images.EmptyTexture}
                                                                                                    stack_box_slot: 
                                                                                                        Padding:=margin{Left:=InterColumnLeftPadding,Top:=InterColumnTopPadding,Right:=InterColumnRightPadding,Bottom:=InterColumnBottomPadding}
                                                                                                        Widget:=overlay:
                                                                                                            Slots:=array:
                                                                                                                overlay_slot:
                                                                                                                    Widget:=RE_Row2_Upgrade_TextBlock
                                                                                                                overlay_slot: 
                                                                                                                    Widget:=texture_block{DefaultDesiredSize:=vector2{X:=SREColumn3SizeX,Y:=SREColumn3SizeY},DefaultImage:=Images.EmptyTexture}
                                                                                                    stack_box_slot: 
                                                                                                        Padding:=margin{Left:=InterColumnLeftPadding,Top:=InterColumnTopPadding,Right:=InterColumnRightPadding,Bottom:=InterColumnBottomPadding}
                                                                                                        Widget:=overlay:
                                                                                                            Slots:=array:   
                                                                                                                overlay_slot: 
                                                                                                                    Widget:=texture_block{DefaultDesiredSize:=vector2{X:=SREColumn4SizeX,Y:=SREColumn4SizeY},DefaultImage:=Images.EmptyTexture} 
                                                                                                                overlay_slot: 
                                                                                                                    HorizontalAlignment:=horizontal_alignment.Left
                                                                                                                    VerticalAlignment:=vertical_alignment.Center 
                                                                                                                    Padding:=margin{Left:=65.0,Top:=0.0,Right:=0.0,Bottom:=0.0}
                                                                                                                    Widget:=RE_Row2_UpgradeCost_TextBlock                
                                                                                                                overlay_slot: 
                                                                                                                    HorizontalAlignment:=horizontal_alignment.Left
                                                                                                                    VerticalAlignment:=vertical_alignment.Center 
                                                                                                                    Padding:=margin{Left:=10.0,Top:=0.0,Right:=0.0,Bottom:=0.0}
                                                                                                                    Widget:=RE_Row2_SapphiresTexture
                                                                                                                overlay_slot:
                                                                                                                    HorizontalAlignment:=horizontal_alignment.Fill
                                                                                                                    VerticalAlignment:=vertical_alignment.Fill  
                                                                                                                    Widget:=RE_Row2Button
                                                                    stack_box_slot: #Line 6
                                                                        Padding:=margin{Left:=RowLeftPadding,Top:=RowTopPadding,Right:=RowRightPadding,Bottom:=RowBottomPadding}
                                                                            Widget:=overlay:
                                                                                Slots:=array:
                                                                                    overlay_slot: 
                                                                                        Widget:=stack_box:
                                                                                            Orientation:=orientation.Horizontal
                                                                                                Slots:=array:
                                                                                                    stack_box_slot: 
                                                                                                        Padding:=margin{Left:=InterColumnLeftPadding,Top:=InterColumnTopPadding,Right:=InterColumnRightPadding,Bottom:=InterColumnBottomPadding}
                                                                                                        Widget:=overlay:
                                                                                                            Slots:=array:
                                                                                                                overlay_slot: 
                                                                                                                    Widget:=texture_block{DefaultDesiredSize:=vector2{X:=SREColumn1SizeX,Y:=SREColumn1SizeY},DefaultImage:=Images.EmptyTexture}
                                                                                                    stack_box_slot: 
                                                                                                        Padding:=margin{Left:=InterColumnLeftPadding,Top:=InterColumnTopPadding,Right:=InterColumnRightPadding,Bottom:=InterColumnBottomPadding}
                                                                                                        Widget:=overlay:
                                                                                                            Slots:=array:
                                                                                                                overlay_slot: 
                                                                                                                    Widget:=texture_block{DefaultDesiredSize:=vector2{X:=SREColumn2SizeX,Y:=SREColumn2SizeY},DefaultImage:=Images.EmptyTexture}
                                                                                                    stack_box_slot: 
                                                                                                        Padding:=margin{Left:=InterColumnLeftPadding,Top:=InterColumnTopPadding,Right:=InterColumnRightPadding,Bottom:=InterColumnBottomPadding}
                                                                                                        Widget:=overlay:
                                                                                                            Slots:=array:
                                                                                                                overlay_slot: 
                                                                                                                    Widget:=texture_block{DefaultDesiredSize:=vector2{X:=SREColumn3SizeX,Y:=SREColumn3SizeY},DefaultImage:=Images.EmptyTexture}
                                                                                                    stack_box_slot: 
                                                                                                        Padding:=margin{Left:=InterColumnLeftPadding,Top:=InterColumnTopPadding,Right:=InterColumnRightPadding,Bottom:=InterColumnBottomPadding}
                                                                                                        Widget:=overlay:
                                                                                                            Slots:=array:   
                                                                                                                overlay_slot: 
                                                                                                                    Widget:=texture_block{DefaultDesiredSize:=vector2{X:=SREColumn4SizeX,Y:=SREColumn4SizeY},DefaultImage:=Images.EmptyTexture}

                                                                    stack_box_slot: #Line 7
                                                                        Padding:=margin{Left:=RowLeftPadding,Top:=RowTopPadding,Right:=RowRightPadding,Bottom:=RowBottomPadding}
                                                                            Widget:=overlay:
                                                                                Slots:=array:
                                                                                    overlay_slot: 
                                                                                        Widget:=stack_box:
                                                                                            Orientation:=orientation.Horizontal
                                                                                                Slots:=array:
                                                                                                    stack_box_slot: 
                                                                                                        Padding:=margin{Left:=InterColumnLeftPadding,Top:=InterColumnTopPadding,Right:=InterColumnRightPadding,Bottom:=InterColumnBottomPadding}
                                                                                                        Widget:=overlay:
                                                                                                            Slots:=array:
                                                                                                                overlay_slot: 
                                                                                                                    Widget:=texture_block{DefaultDesiredSize:=vector2{X:=SREColumn1SizeX,Y:=SREColumn1SizeY},DefaultImage:=Images.EmptyTexture}
                                                                                                    stack_box_slot: 
                                                                                                        Padding:=margin{Left:=InterColumnLeftPadding,Top:=InterColumnTopPadding,Right:=InterColumnRightPadding,Bottom:=InterColumnBottomPadding}
                                                                                                        Widget:=overlay:
                                                                                                            Slots:=array:
                                                                                                                overlay_slot: 
                                                                                                                    Widget:=texture_block{DefaultDesiredSize:=vector2{X:=SREColumn2SizeX,Y:=SREColumn2SizeY},DefaultImage:=Images.EmptyTexture}
                                                                                                    stack_box_slot: 
                                                                                                        Padding:=margin{Left:=InterColumnLeftPadding,Top:=InterColumnTopPadding,Right:=InterColumnRightPadding,Bottom:=InterColumnBottomPadding}
                                                                                                        Widget:=overlay:
                                                                                                            Slots:=array:
                                                                                                                overlay_slot: 
                                                                                                                    Widget:=texture_block{DefaultDesiredSize:=vector2{X:=SREColumn3SizeX,Y:=SREColumn3SizeY},DefaultImage:=Images.EmptyTexture}
                                                                                                    stack_box_slot: 
                                                                                                        Padding:=margin{Left:=InterColumnLeftPadding,Top:=InterColumnTopPadding,Right:=InterColumnRightPadding,Bottom:=InterColumnBottomPadding}
                                                                                                        Widget:=overlay:
                                                                                                            Slots:=array:   
                                                                                                                overlay_slot: 
                                                                                                                    Widget:=texture_block{DefaultDesiredSize:=vector2{X:=SREColumn4SizeX,Y:=SREColumn4SizeY},DefaultImage:=Images.EmptyTexture}         
                                                                    stack_box_slot: #Line 8
                                                                        Padding:=margin{Left:=RowLeftPadding,Top:=RowTopPadding,Right:=RowRightPadding,Bottom:=RowBottomPadding}
                                                                            Widget:=overlay:
                                                                                Slots:=array:
                                                                                    overlay_slot:
                                                                                        HorizontalAlignment:=horizontal_alignment.Center
                                                                                        VerticalAlignment:=vertical_alignment.Center
                                                                                        Widget:=WarningTitle_TextBlock
                                                                                    
                                                                                        <#
                                                                                        Widget:=stack_box:
                                                                                            Orientation:=orientation.Horizontal
                                                                                                Slots:=array:
                                                                                                    stack_box_slot: 
                                                                                                        Padding:=margin{Left:=InterColumnLeftPadding,Top:=InterColumnTopPadding,Right:=InterColumnRightPadding,Bottom:=InterColumnBottomPadding}
                                                                                                        Widget:=overlay:
                                                                                                            Slots:=array:
                                                                                                                overlay_slot: 
                                                                                                                    Widget:=texture_block{DefaultDesiredSize:=vector2{X:=SREColumn1SizeX,Y:=SREColumn1SizeY},DefaultImage:=Images.EmptyTexture}
                                                                                                    stack_box_slot: 
                                                                                                        Padding:=margin{Left:=InterColumnLeftPadding,Top:=InterColumnTopPadding,Right:=InterColumnRightPadding,Bottom:=InterColumnBottomPadding}
                                                                                                        Widget:=overlay:
                                                                                                            Slots:=array:
                                                                                                                overlay_slot: 
                                                                                                                    Widget:=texture_block{DefaultDesiredSize:=vector2{X:=SREColumn2SizeX,Y:=SREColumn2SizeY},DefaultImage:=Images.EmptyTexture}
                                                                                                    stack_box_slot: 
                                                                                                        Padding:=margin{Left:=InterColumnLeftPadding,Top:=InterColumnTopPadding,Right:=InterColumnRightPadding,Bottom:=InterColumnBottomPadding}
                                                                                                        Widget:=overlay:
                                                                                                            Slots:=array:
                                                                                                                overlay_slot: 
                                                                                                                    Widget:=texture_block{DefaultDesiredSize:=vector2{X:=SREColumn3SizeX,Y:=SREColumn3SizeY},DefaultImage:=Images.EmptyTexture}
                                                                                                    stack_box_slot: 
                                                                                                        Padding:=margin{Left:=InterColumnLeftPadding,Top:=InterColumnTopPadding,Right:=InterColumnRightPadding,Bottom:=InterColumnBottomPadding}
                                                                                                        Widget:=overlay:
                                                                                                            Slots:=array:   
                                                                                                                overlay_slot: 
                                                                                                                    Widget:=texture_block{DefaultDesiredSize:=vector2{X:=SREColumn4SizeX,Y:=SREColumn4SizeY},DefaultImage:=Images.EmptyTexture}#>         
                                                                    stack_box_slot: #Line 9
                                                                        Padding:=margin{Left:=RowLeftPadding,Top:=RowTopPadding,Right:=RowRightPadding,Bottom:=RowBottomPadding}
                                                                            Widget:=overlay:
                                                                                Slots:=array:
                                                                                    #overlay_slot: 
                                                                                    overlay_slot:
                                                                                        HorizontalAlignment:=horizontal_alignment.Center
                                                                                        VerticalAlignment:=vertical_alignment.Bottom
                                                                                        Widget:=Warning_TextBlock
                                                                                        <#
                                                                                        Widget:=stack_box:
                                                                                            Orientation:=orientation.Horizontal
                                                                                                Slots:=array:
                                                                                                    stack_box_slot: 
                                                                                                        Padding:=margin{Left:=InterColumnLeftPadding,Top:=InterColumnTopPadding,Right:=InterColumnRightPadding,Bottom:=InterColumnBottomPadding}
                                                                                                        Widget:=overlay:
                                                                                                            Slots:=array:
                                                                                                                overlay_slot: 
                                                                                                                    Widget:=texture_block{DefaultDesiredSize:=vector2{X:=SREColumn1SizeX,Y:=SREColumn1SizeY},DefaultImage:=Images.EmptyTexture}
                                                                                                    stack_box_slot: 
                                                                                                        Padding:=margin{Left:=InterColumnLeftPadding,Top:=InterColumnTopPadding,Right:=InterColumnRightPadding,Bottom:=InterColumnBottomPadding}
                                                                                                        Widget:=overlay:
                                                                                                            Slots:=array:
                                                                                                                overlay_slot: 
                                                                                                                    Widget:=texture_block{DefaultDesiredSize:=vector2{X:=SREColumn2SizeX,Y:=SREColumn2SizeY},DefaultImage:=Images.EmptyTexture}
                                                                                                    stack_box_slot: 
                                                                                                        Padding:=margin{Left:=InterColumnLeftPadding,Top:=InterColumnTopPadding,Right:=InterColumnRightPadding,Bottom:=InterColumnBottomPadding}
                                                                                                        Widget:=overlay:
                                                                                                            Slots:=array:
                                                                                                                overlay_slot: 
                                                                                                                    Widget:=texture_block{DefaultDesiredSize:=vector2{X:=SREColumn3SizeX,Y:=SREColumn3SizeY},DefaultImage:=Images.EmptyTexture}
                                                                                                    stack_box_slot: 
                                                                                                        Padding:=margin{Left:=InterColumnLeftPadding,Top:=InterColumnTopPadding,Right:=InterColumnRightPadding,Bottom:=InterColumnBottomPadding}
                                                                                                        Widget:=overlay:
                                                                                                            Slots:=array:   
                                                                                                                overlay_slot: 
                                                                                                                    Widget:=texture_block{DefaultDesiredSize:=vector2{X:=SREColumn4SizeX,Y:=SREColumn4SizeY},DefaultImage:=Images.EmptyTexture}
                                                                                                                        #>

                                    stack_box_slot:
                                        HorizontalAlignment:=horizontal_alignment.Center
                                        VerticalAlignment:=vertical_alignment.Center
                                        Padding:=margin{Top:=20.0}
                                        Widget:=overlay:
                                            Slots:=array:
                                                overlay_slot: 
                                                    Widget:=texture_block{DefaultDesiredSize:=vector2{X:=1800.0,Y:=80.0},DefaultImage:=Images.EmptyTexture}
                                                overlay_slot:
                                                    HorizontalAlignment:=horizontal_alignment.Fill
                                                    VerticalAlignment:=vertical_alignment.Fill
                                                    Widget:=CloseButton                      
                <#Warning message
                canvas_slot:
                    Anchors:=anchors{Minimum:=vector2{X:=0.5,Y:=0.5},Maximum:=vector2{X:=0.5,Y:=0.5}}
                    ZOrder:=2
                    SizeToContent:=true
                    Offsets:=margin{Left:=0.0,Top:=0.0,Right:=0.0,Bottom:=0.0} #PositionX, PositionY, SizeX, SizeY
                    Alignment:=vector2{X:=0.5,Y:=0.0}
                    Widget:=overlay:
                        Slots:=array:
                            overlay_slot:
                                HorizontalAlignment:=horizontal_alignment.Center
                                VerticalAlignment:=vertical_alignment.Center
                                Widget:=overlay:
                                    Slots:=array:
                                        overlay_slot:
                                            Widget:=Warning_Slider_Texture
                            overlay_slot:
                                HorizontalAlignment:=horizontal_alignment.Center
                                VerticalAlignment:=vertical_alignment.Bottom
                                Widget:=overlay:
                                    Slots:=array:
                                        overlay_slot:
                                            HorizontalAlignment:=horizontal_alignment.Fill
                                            VerticalAlignment:=vertical_alignment.Fill
                                            Widget:=texture_block{DefaultImage:=Images.BlackBackground, DefaultDesiredSize:=vector2{X:=1.0,Y:=1.0}}
                                        overlay_slot:
                                            HorizontalAlignment:=horizontal_alignment.Fill
                                            VerticalAlignment:=vertical_alignment.Fill
                                            Padding:=margin{Left:=2.5,Top:=2.5,Right:=2.5,Bottom:=2.5}
                                            Widget:=texture_block{DefaultImage:=Images.BlackBackground, DefaultDesiredSize:=vector2{X:=1.0,Y:=1.0}}
                                        overlay_slot: #Screw 0
                                            Padding:=margin{Left:=3.5,Top:=3.5,Right:=3.5,Bottom:=3.5}
                                            HorizontalAlignment:=horizontal_alignment.Left
                                            VerticalAlignment:=vertical_alignment.Top
                                            Widget:=texture_block{DefaultImage:=Images.Dialogue_Bolt, DefaultDesiredSize:=vector2{X:=5.0,Y:=5.0}}
                                        overlay_slot: #Screw 1
                                            Padding:=margin{Left:=3.5,Top:=3.5,Right:=3.5,Bottom:=3.5}
                                            HorizontalAlignment:=horizontal_alignment.Right
                                            VerticalAlignment:=vertical_alignment.Top
                                            Widget:=texture_block{DefaultImage:=Images.Dialogue_Bolt, DefaultDesiredSize:=vector2{X:=5.0,Y:=5.0}}
                                        overlay_slot: #Screw 2
                                            Padding:=margin{Left:=3.5,Top:=3.5,Right:=3.5,Bottom:=3.5}
                                            HorizontalAlignment:=horizontal_alignment.Left
                                            VerticalAlignment:=vertical_alignment.Bottom
                                            Widget:=texture_block{DefaultImage:=Images.Dialogue_Bolt, DefaultDesiredSize:=vector2{X:=5.0,Y:=5.0}}
                                        overlay_slot: #Screw 3
                                            Padding:=margin{Left:=3.5,Top:=3.5,Right:=3.5,Bottom:=3.5}
                                            HorizontalAlignment:=horizontal_alignment.Right
                                            VerticalAlignment:=vertical_alignment.Bottom
                                            Widget:=texture_block{DefaultImage:=Images.Dialogue_Bolt, DefaultDesiredSize:=vector2{X:=5.0,Y:=5.0}}
                                        overlay_slot:
                                            HorizontalAlignment:=horizontal_alignment.Center
                                            VerticalAlignment:=vertical_alignment.Center
                                            Padding:=margin{Left:=RowLeftPadding,Top:=RowTopPadding,Right:=RowRightPadding,Bottom:=RowBottomPadding}
                                            Widget:=stack_box:
                                                Orientation:=orientation.Vertical
                                                    Slots:=array:
                                                        stack_box_slot:
                                                            Padding:=margin{Left:=5.0,Top:=5.0,Right:=5.0,Bottom:=5.0}
                                                            HorizontalAlignment:=horizontal_alignment.Center
                                                            VerticalAlignment:=vertical_alignment.Center
                                                            Widget:=text_block{DefaultText:=BasicMessage("Warning"), DefaultTextColor:=Yellow, DefaultShadowOffset:=ShadowOffset, DefaultShadowOpacity := DefaultTextShadowOpacity}
                                                        stack_box_slot:
                                                            Padding:=margin{Left:=5.0,Top:=2.5,Right:=5.0,Bottom:=2.5}
                                                            HorizontalAlignment:=horizontal_alignment.Center
                                                            VerticalAlignment:=vertical_alignment.Center
                                                            Widget:=Warning_TextBlock
                                                        stack_box_slot:
                                                            Padding:=margin{Left:=5.0,Top:=5.0,Right:=5.0,Bottom:=5.0}
                                                            HorizontalAlignment:=horizontal_alignment.Center
                                                            VerticalAlignment:=vertical_alignment.Center
                                                            Widget:=text_block{DefaultText:=BasicMessage("You should improve your pickaxe damage first"), DefaultTextColor:=Orange, DefaultShadowOffset:=ShadowOffset, DefaultShadowOpacity := DefaultTextShadowOpacity}#>
                        
MakeBBakerUpgradeUI<constructor>(
    InBakeryMachine : bakery_machines_manager)<transacts> := Bbaker_upgrades_ui :
        BakeryMachinesManager := InBakeryMachine
               