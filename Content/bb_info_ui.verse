using { /Verse.org/Colors/NamedColors }
using { /Verse.org/Colors }
using { /UnrealEngine.com/Temporary/UI }
using { /Verse.org/Simulation }
using { /UnrealEngine.com/Temporary/SpatialMath }
using { /Fortnite.com/UI }

# See https://dev.epicgames.com/documentation/en-us/uefn/create-your-own-device-in-verse for how to create a verse device.

# A Verse-authored creative device that can be placed in a level
bb_info_ui := class():

    #Class settings
    var BakeryMachinesManager : bakery_machines_manager

    #Background
    var Background_Texture : texture_block = texture_block{DefaultDesiredSize:=vector2{X:=2000.0,Y:=1250.0},DefaultImage:=Images.Img_MainBackground}

    #Title
    BakerIcon_Texture : texture_block = texture_block{DefaultDesiredSize:=vector2{X:=100.0,Y:=100.0},DefaultImage:=Images.Baker_Icon}
    BakerTitle_TextBlock : text_block = text_block{DefaultTextColor:=White, DefaultShadowOffset:=ShadowOffset, DefaultShadowOpacity := 1.0}
    HorizontalLine : texture_block = texture_block{DefaultDesiredSize:=vector2{X:=HorizontalLineSizeX,Y:=HorizontalLineSizeY},DefaultImage:=Images.Img_HorizontalLine}

    #Top Info
    ProductionInfo_TextBlock : text_block = text_block{DefaultTextColor:=White, DefaultShadowOffset:=ShadowOffset, DefaultShadowOpacity := 1.0}
    SpeedInfo_TextBlock : text_block = text_block{DefaultTextColor:=White, DefaultShadowOffset:=ShadowOffset, DefaultShadowOpacity := 1.0}
    SliderBatchInfoBackground_EmptyTexture : texture_block = texture_block{DefaultImage:=Images.EmptyTexture,DefaultDesiredSize:=vector2{X:=FullProgressBakerInfoUI_X,Y:=60.0}}
    SliderBatchInfoProgress_Texture : texture_block = texture_block{DefaultImage:=Images.Slider_Progress,DefaultTint:=LightGreen}
    SliderBatchInfoBackground_Texture : texture_block = texture_block{DefaultImage:=Images.Slider_Progress, DefaultDesiredSize:=vector2{X:=FullProgressBakerInfoUI_X,Y:=60.0}}
    SliderBatchProgress_TextBlock : text_block = text_block{DefaultTextColor:=White, DefaultShadowOffset:=ShadowOffset, DefaultShadowOpacity := 1.0}

    #Water
    WaterAmmount_TextBlock : text_block = text_block{DefaultTextColor:=WaterTextColor, DefaultShadowOffset:=ShadowOffset, DefaultShadowOpacity := 1.0}
    WaterSlider_EmptyTexture : texture_block = texture_block{DefaultImage:=Images.EmptyTexture, DefaultDesiredSize:=vector2{Y:=450.0}}
    WaterIcon_Texture : texture_block = texture_block{DefaultImage:=Images.Water_Icon, DefaultDesiredSize:=vector2{X:=150.0,Y:=150.0}}
    WaterSliderProgress_Texture : texture_block = texture_block{DefaultImage:=Images.Slider_Progress, DefaultTint := InfoBakerWaterColor,DefaultDesiredSize:=vector2{X:=0.0,Y:=0.0}}
    WaterSliderBackground_Texture : texture_block = texture_block{DefaultImage:=Images.Img_LineBackground, DefaultDesiredSize:=vector2{X:=FullBakerInfoUI_DepositX,Y:=FullBakerInfoUI_DepositY}}
    WaterEfficiency_TextBlock : text_block = text_block{DefaultTextColor:=White, DefaultShadowOffset:=ShadowOffset, DefaultShadowOpacity := 1.0}

    #Flour
    FlourAmmount_TextBlock : text_block = text_block{DefaultTextColor:=FlourTextColor, DefaultShadowOffset:=ShadowOffset, DefaultShadowOpacity := 1.0}
    FlourSlider_EmptyTexture : texture_block = texture_block{DefaultImage:=Images.EmptyTexture, DefaultDesiredSize:=vector2{Y:=450.0}}
    FlourIcon_Texture : texture_block = texture_block{DefaultImage:=Images.Flour_Icon, DefaultDesiredSize:=vector2{X:=150.0,Y:=150.0}}
    FlourSliderProgress_Texture : texture_block = texture_block{DefaultImage:=Images.Slider_Progress, DefaultTint := InfoBakerFlourColor,DefaultDesiredSize:=vector2{X:=0.0,Y:=0.0}}
    FlourSliderBackground_Texture : texture_block = texture_block{DefaultImage:=Images.Img_LineBackground, DefaultDesiredSize:=vector2{X:=FullBakerInfoUI_DepositX,Y:=FullBakerInfoUI_DepositY}}
    FlourEfficiency_TextBlock : text_block = text_block{DefaultTextColor:=White, DefaultShadowOffset:=ShadowOffset, DefaultShadowOpacity := 1.0}

    var MaybeCanvas : ?canvas = false
    var UIIsDisplayed : logic = false

    var TimeElapsed : int = 0

    #Close
    CloseButton : button_loud = button_loud{}

    InitUI() : void =
        InitSubscribers()

        BakerTitle_TextBlock.SetText(BasicMessage("Bread baker production info"))
        CloseButton.SetText(BasicMessage("OK"))

        Canvas := MakeCanvas()
        set MaybeCanvas = option{Canvas}
        RemoveDelay_OnRemove()

    InitSubscribers() : void =
        CloseButton.OnClick().Subscribe(Remove_Canvas)

    SetResourceSlider(Ammount : float, Limit : float, Type : int) : void =
        var Percentage : float = Ammount / Limit
        if(Percentage>1.0):
            set Percentage = 1.0
        if(Type = WATER):
            WaterAmmount_TextBlock.SetText(BasicMessage("{GetOptimizedMagnitudeLetter(Ammount)}\n/ {GetOptimizedMagnitudeLetter(Limit)}"))
            WaterSliderProgress_Texture.SetDesiredSize(vector2{X:=10.0,Y:=FullBakerInfoUI_DepositY*Percentage})
        else if(Type = FLOUR):
            FlourAmmount_TextBlock.SetText(BasicMessage("{GetOptimizedMagnitudeLetter(Ammount)}\n/ {GetOptimizedMagnitudeLetter(Limit)}"))
            FlourSliderProgress_Texture.SetDesiredSize(vector2{X:=10.0,Y:=FullBakerInfoUI_DepositY*Percentage})
    
    SetProgressForB(Progress_ForB : float, BachProduction : float) : void =
        SliderBatchInfoProgress_Texture.SetDesiredSize(vector2{X:=FullProgressBakerInfoUI_X*Progress_ForB})
        SliderBatchProgress_TextBlock.SetText(BasicMessage("Current Production: {GetOptimizedMagnitudeLetter(BachProduction)} units | Batch progress: {GetOptimizedMagnitudeLetter(Progress_ForB*100.0)}%"))

    UpdateUI() : void =
        Print("TO DO UPDATE UI")

    Remove_Canvas(Message : widget_message) : void =
        set UIIsDisplayed = false
        if:
            Agent:=BakeryMachinesManager.MaybeAgentOwner?
            Player:=player[Agent]
            PlayerUI:=GetPlayerUI[Player]
            Canvas := MaybeCanvas?
        then:
            set BakeryMachinesManager.BBInfoIsDisplayed = false
            PlayerUI.RemoveWidget(Canvas)
            RemoveDelay_OnRemove()
            BakeryMachinesManager.CheckQuestProgress(1.0,CHECKINFO_TUTORIAL_TYPE)
            if(BakeryMachinesManager.IsOnUKTutorial?, UKTutorial := BakeryMachinesManager.MaybeUKTutorial?):
                UKTutorial.InfoCheckedEvent()

    Remove_CanvasWithoutMessage() : void =
        set UIIsDisplayed = false
        if:
            Agent:=BakeryMachinesManager.MaybeAgentOwner?
            Player:=player[Agent]
            PlayerUI:=GetPlayerUI[Player]
            Canvas := MaybeCanvas?
        then:
            set BakeryMachinesManager.BBInfoIsDisplayed = false
            PlayerUI.RemoveWidget(Canvas)
            RemoveDelay_OnRemove()
            BakeryMachinesManager.CheckQuestProgress(1.0,CHECKINFO_TUTORIAL_TYPE)
            if(BakeryMachinesManager.IsOnUKTutorial?, UKTutorial := BakeryMachinesManager.MaybeUKTutorial?):
                UKTutorial.InfoCheckedEvent()


    SetProductionInfo(Production : float, Speed : float) : void =
        SpeedInfo_TextBlock.SetText(BasicMessage("Baking time: {GetOptimizedMagnitudeLetter(Speed/BakeryMachinesManager.SuperEnergy_Multiplier)} s / Batch"))
        ProductionInfo_TextBlock.SetText(BasicMessage("Next Batch Production: {GetOptimizedMagnitudeLetter(Production*BakeryMachinesManager.SuperEnergy_Multiplier)} units / Batch"))
            
    Display_Canvas(AllEfficiency:[]float, TotalProduction : float, Speed : float) : void =
        set TimeElapsed = 0
        set UIIsDisplayed = true

        if:
            B_WaterEfficiency := AllEfficiency[WATER]-1.0
            B_FlourEfficiency := AllEfficiency[FLOUR]-1.0
        then:
            WaterEfficiency_TextBlock.SetText(BasicMessage("Efficiency {ToTruncatedFloat(B_WaterEfficiency*100.0,1)}%"))
            FlourEfficiency_TextBlock.SetText(BasicMessage("Efficiency {ToTruncatedFloat(B_FlourEfficiency*100.0,1)}%"))
            SpeedInfo_TextBlock.SetText(BasicMessage("Baking time: {GetOptimizedMagnitudeLetter(Speed/BakeryMachinesManager.SuperEnergy_Multiplier)} s / Batch"))
            ProductionInfo_TextBlock.SetText(BasicMessage("Next Batch Production: {GetOptimizedMagnitudeLetter(TotalProduction*BakeryMachinesManager.SuperEnergy_Multiplier)} units / Batch"))

        var Index : int = 0
        if:
            Agent:=BakeryMachinesManager.MaybeAgentOwner?
            PlayerUI:=GetPlayerUI[player[Agent]]
            Canvas := MaybeCanvas?
        then:
            PlayerUI.AddWidget(Canvas,player_ui_slot{InputMode := ui_input_mode.All, ZOrder:=1})
            RemoveDelay_OnAdd()   

    StartUIAnimation(Index : int, Ammount : float)<suspends> : void =

        #Print("Starting animation with index: {Index}")

        
        FramesPerAction_2 : int = 10

        #Phase1
        FramesPerAction_1 : int = 5
        AddX : float = ActionIconSizeX/(FramesPerAction_1*1.0)
        AddY : float = ActionIconSizeY/(FramesPerAction_1*1.0)
        SmallAmmount:=Ammount/(FramesPerAction_1*1.0)
        var SizeVector : vector2 = vector2{X:=AddX,Y:=AddY}
        var TotalVector : vector2 = vector2{X:=0.0,Y:=0.0}
   

    RemoveDelay_OnAdd() : void =
        BakerTitle_TextBlock.SetText(BasicMessage("Bread baker production info"))
        BakerIcon_Texture.SetVisibility(widget_visibility.Visible)

    RemoveDelay_OnRemove() : void =
        BakerTitle_TextBlock.SetText(BasicMessage("Bread baker production info "))
        BakerIcon_Texture.SetVisibility(widget_visibility.Hidden)


    MakeCanvas(): canvas=
        Canvas := canvas:
            Slots:=array:
                #Main overlay
                canvas_slot:
                    Anchors:=anchors{Minimum:=vector2{X:=0.5,Y:=0.5},Maximum:=vector2{X:=0.5,Y:=0.5}}
                    ZOrder:=0
                    Offsets:=margin{Left:=0.0,Top:=0.0,Right:=0.0,Bottom:=0.0} #PositionX, PositionY, SizeX, SizeY
                    Alignment:=vector2{X:=0.5,Y:=0.5}
                    Widget:=Background_Texture
                canvas_slot:
                    Anchors:=anchors{Minimum:=vector2{X:=0.5,Y:=0.5},Maximum:=vector2{X:=0.5,Y:=0.5}}
                    ZOrder:=0
                    Offsets:=margin{Left:=0.0,Top:=0.0,Right:=0.0,Bottom:=0.0} #PositionX, PositionY, SizeX, SizeY
                    Alignment:=vector2{X:=0.5,Y:=0.5}
                    Widget:=overlay:
                        Slots:=array:
                            overlay_slot:
                                Widget:=stack_box: 
                                    Orientation:=orientation.Vertical
                                        Slots:=array:
                                            stack_box_slot: #Baker Icon
                                                Padding:=margin{Left:=RowLeftPadding,Top:=RowTopPadding,Right:=RowRightPadding,Bottom:=RowBottomPadding}
                                                Widget:=overlay:
                                                    Slots:=array:
                                                        overlay_slot:
                                                            Widget:=BakerIcon_Texture
                                            stack_box_slot: #Title
                                                Padding:=margin{Left:=RowLeftHugePadding,Top:=RowTopPadding,Right:=RowRightHugePadding,Bottom:=RowBottomHugePadding}
                                                Widget:=overlay:
                                                    Slots:=array:
                                                        overlay_slot: 
                                                            HorizontalAlignment:=horizontal_alignment.Fill
                                                            VerticalAlignment:=vertical_alignment.Bottom
                                                            Widget:=HorizontalLine       
                                                        overlay_slot: 
                                                            HorizontalAlignment:=horizontal_alignment.Center
                                                            VerticalAlignment:=vertical_alignment.Center
                                                            Widget:=texture_block{DefaultImage:=Images.EmptyTexture, DefaultDesiredSize:=vector2{X:=250.0,Y:=80.0}}
                                                        overlay_slot:
                                                            HorizontalAlignment:=horizontal_alignment.Center
                                                            VerticalAlignment:=vertical_alignment.Center
                                                            Widget:=BakerTitle_TextBlock
                                            stack_box_slot: #Progress Info
                                                Padding:=margin{Left:=RowLeftHugePadding,Top:=RowTopHugePadding,Right:=RowRightHugePadding,Bottom:=RowBottomHugePadding}
                                                Widget:=overlay: 
                                                    Slots:=array:
                                                        overlay_slot: #Baker Icon
                                                            Widget:=stack_box: 
                                                                Orientation:=orientation.Vertical
                                                                    Slots:=array:
                                                                        stack_box_slot:
                                                                                Widget:=stack_box: 
                                                                                    Orientation:=orientation.Vertical
                                                                                        Slots:=array:
                                                                                            stack_box_slot: #Production
                                                                                                Padding:=margin{Left:=RowLeftPadding,Top:=RowTopPadding,Right:=RowRightPadding,Bottom:=RowBottomPadding}
                                                                                                HorizontalAlignment:=horizontal_alignment.Fill
                                                                                                VerticalAlignment:=vertical_alignment.Fill
                                                                                                Widget:=overlay: 
                                                                                                    Slots:=array:
                                                                                                        overlay_slot:
                                                                                                            HorizontalAlignment:=horizontal_alignment.Fill
                                                                                                            VerticalAlignment:=vertical_alignment.Fill
                                                                                                            Widget:=texture_block{DefaultImage:=Images.Img_LineBackground}
                                                                                                        overlay_slot:
                                                                                                            HorizontalAlignment:=horizontal_alignment.Center
                                                                                                            Widget:=stack_box: 
                                                                                                                Orientation:=orientation.Horizontal
                                                                                                                    Slots:=array:
                                                                                                                        stack_box_slot:
                                                                                                                            HorizontalAlignment:=horizontal_alignment.Center
                                                                                                                            VerticalAlignment:=vertical_alignment.Center
                                                                                                                            Padding:=margin{Left:=RowLeftPadding,Top:=RowTopPadding,Right:=RowRightPadding,Bottom:=RowBottomPadding}
                                                                                                                            Widget:=texture_block{DefaultImage:=Images.Baker_Icon,DefaultDesiredSize:=vector2{X:=SizeSmallIconX,Y:=SizeSmallIconY}}
                                                                                                                        stack_box_slot:
                                                                                                                            HorizontalAlignment:=horizontal_alignment.Left
                                                                                                                            VerticalAlignment:=vertical_alignment.Center
                                                                                                                            Padding:=margin{Left:=RowLeftPadding,Top:=RowTopPadding,Right:=RowRightPadding,Bottom:=RowBottomPadding}
                                                                                                                            Widget:=ProductionInfo_TextBlock
                                                                                            stack_box_slot: #Speed
                                                                                                Padding:=margin{Left:=RowLeftPadding,Top:=RowTopPadding,Right:=RowRightPadding,Bottom:=RowBottomPadding}
                                                                                                HorizontalAlignment:=horizontal_alignment.Fill
                                                                                                VerticalAlignment:=vertical_alignment.Fill
                                                                                                Widget:=overlay: 
                                                                                                    Slots:=array:
                                                                                                        overlay_slot:
                                                                                                            HorizontalAlignment:=horizontal_alignment.Fill
                                                                                                            VerticalAlignment:=vertical_alignment.Fill
                                                                                                            Widget:=texture_block{DefaultImage:=Images.Img_LineBackground}
                                                                                                        overlay_slot:
                                                                                                            HorizontalAlignment:=horizontal_alignment.Center
                                                                                                            Widget:=stack_box: 
                                                                                                                Orientation:=orientation.Horizontal
                                                                                                                    Slots:=array:
                                                                                                                        stack_box_slot:
                                                                                                                            HorizontalAlignment:=horizontal_alignment.Center
                                                                                                                            VerticalAlignment:=vertical_alignment.Center
                                                                                                                            Padding:=margin{Left:=RowLeftPadding,Top:=RowTopPadding,Right:=RowRightPadding,Bottom:=RowBottomPadding}
                                                                                                                            Widget:=texture_block{DefaultImage:=Images.Speed_Icon,DefaultDesiredSize:=vector2{X:=SizeSmallIconX,Y:=SizeSmallIconY}}
                                                                                                                        stack_box_slot:
                                                                                                                            HorizontalAlignment:=horizontal_alignment.Left
                                                                                                                            VerticalAlignment:=vertical_alignment.Center
                                                                                                                            Padding:=margin{Left:=RowLeftPadding,Top:=RowTopPadding,Right:=RowRightPadding,Bottom:=RowBottomPadding}
                                                                                                                            Widget:=SpeedInfo_TextBlock
                                                                                            stack_box_slot: #ProgressBar
                                                                                                Padding:=margin{Left:=RowLeftHugePadding,Top:=RowTopHugePadding,Right:=RowRightHugePadding,Bottom:=RowBottomHugePadding}
                                                                                                HorizontalAlignment:=horizontal_alignment.Center
                                                                                                VerticalAlignment:=vertical_alignment.Fill
                                                                                                Widget:=overlay: 
                                                                                                    Slots:=array:
                                                                                                        overlay_slot:
                                                                                                            HorizontalAlignment:=horizontal_alignment.Center
                                                                                                            VerticalAlignment:=vertical_alignment.Center
                                                                                                            Widget:=SliderBatchInfoBackground_EmptyTexture
                                                                                                        overlay_slot:
                                                                                                            HorizontalAlignment:=horizontal_alignment.Left
                                                                                                            VerticalAlignment:=vertical_alignment.Center
                                                                                                            Widget:=SliderBatchInfoBackground_Texture
                                                                                                        overlay_slot:
                                                                                                            HorizontalAlignment:=horizontal_alignment.Left
                                                                                                            VerticalAlignment:=vertical_alignment.Fill
                                                                                                            Widget:=SliderBatchInfoProgress_Texture
                                                                                                        overlay_slot:
                                                                                                            HorizontalAlignment:=horizontal_alignment.Center
                                                                                                            VerticalAlignment:=vertical_alignment.Center
                                                                                                            Widget:=SliderBatchProgress_TextBlock
                                            stack_box_slot: #Fill deposits
                                                Padding:=margin{Left:=RowLeftPadding,Top:=RowTopPadding,Right:=RowRightPadding,Bottom:=RowBottomPadding}
                                                Widget:=stack_box: 
                                                    Orientation:=orientation.Horizontal
                                                        Slots:=array:
                                                            stack_box_slot: #Water
                                                                Padding:=margin{Left:=RowLeftPadding,Top:=RowTopPadding,Right:=RowRightPadding,Bottom:=RowBottomPadding}
                                                                Widget:=overlay: 
                                                                    Slots:=array:
                                                                        overlay_slot: 
                                                                            Widget:=stack_box: 
                                                                                Orientation:=orientation.Vertical
                                                                                    Slots:=array:
                                                                                        stack_box_slot:
                                                                                            Widget:=overlay: 
                                                                                                Slots:=array:
                                                                                                    overlay_slot: 
                                                                                                        Widget:=WaterSliderBackground_Texture
                                                                                                    overlay_slot: 
                                                                                                        HorizontalAlignment:=horizontal_alignment.Fill
                                                                                                        VerticalAlignment:=vertical_alignment.Bottom
                                                                                                        Widget:=WaterSliderProgress_Texture
                                                                                                    overlay_slot: 
                                                                                                        Padding:=margin{Left:=0.0,Top:=0.0,Right:=0.0,Bottom:=70.0}
                                                                                                        HorizontalAlignment:=horizontal_alignment.Center
                                                                                                        VerticalAlignment:=vertical_alignment.Bottom
                                                                                                        Widget:=WaterAmmount_TextBlock
                                                                                                    overlay_slot: 
                                                                                                        Padding:=margin{Left:=0.0,Top:=75.0,Right:=0.0,Bottom:=0.0}
                                                                                                        HorizontalAlignment:=horizontal_alignment.Center
                                                                                                        VerticalAlignment:=vertical_alignment.Top
                                                                                                        Widget:=WaterIcon_Texture
                                                                                        stack_box_slot:
                                                                                            Padding:=margin{Left:=RowLeftPadding,Top:=RowTopHugePadding,Right:=RowRightPadding,Bottom:=RowBottomHugePadding}
                                                                                            Widget:=overlay: 
                                                                                                Slots:=array:
                                                                                                    overlay_slot: 
                                                                                                        Widget:=texture_block{DefaultDesiredSize:=vector2{X:=FullBakerInfoUI_DepositX},DefaultImage:=Images.EmptyTexture}
                                                                                                    overlay_slot: 
                                                                                                        HorizontalAlignment:=horizontal_alignment.Fill
                                                                                                        VerticalAlignment:=vertical_alignment.Fill
                                                                                                        Widget:=texture_block{DefaultImage:=Images.Img_LineBackground}
                                                                                                    overlay_slot: 
                                                                                                        Padding:=margin{Left:=RowLeftPadding,Top:=RowTopPadding,Right:=RowRightPadding,Bottom:=RowBottomPadding}
                                                                                                        HorizontalAlignment:=horizontal_alignment.Center
                                                                                                        VerticalAlignment:=vertical_alignment.Center
                                                                                                        Widget:=WaterEfficiency_TextBlock
                                                            stack_box_slot: #Flour
                                                                Padding:=margin{Left:=RowLeftPadding,Top:=RowTopPadding,Right:=RowRightPadding,Bottom:=RowBottomPadding}
                                                                Widget:=overlay: 
                                                                    Slots:=array:
                                                                        overlay_slot: 
                                                                            Widget:=stack_box: 
                                                                                Orientation:=orientation.Vertical
                                                                                    Slots:=array:
                                                                                        stack_box_slot:
                                                                                            Widget:=overlay: 
                                                                                                Slots:=array:
                                                                                                    overlay_slot: 
                                                                                                        Widget:=FlourSliderBackground_Texture
                                                                                                    overlay_slot: 
                                                                                                        HorizontalAlignment:=horizontal_alignment.Fill
                                                                                                        VerticalAlignment:=vertical_alignment.Bottom
                                                                                                        Widget:=FlourSliderProgress_Texture
                                                                                                    overlay_slot: 
                                                                                                        Padding:=margin{Left:=0.0,Top:=0.0,Right:=0.0,Bottom:=70.0}
                                                                                                        HorizontalAlignment:=horizontal_alignment.Center
                                                                                                        VerticalAlignment:=vertical_alignment.Bottom
                                                                                                        Widget:=FlourAmmount_TextBlock
                                                                                                    overlay_slot: 
                                                                                                        Padding:=margin{Left:=0.0,Top:=75.0,Right:=0.0,Bottom:=0.0}
                                                                                                        HorizontalAlignment:=horizontal_alignment.Center
                                                                                                        VerticalAlignment:=vertical_alignment.Top
                                                                                                        Widget:=FlourIcon_Texture
                                                                                        stack_box_slot:
                                                                                            Padding:=margin{Left:=RowLeftPadding,Top:=RowTopHugePadding,Right:=RowRightPadding,Bottom:=RowBottomHugePadding}
                                                                                            Widget:=overlay: 
                                                                                                Slots:=array:
                                                                                                    overlay_slot: 
                                                                                                        Widget:=texture_block{DefaultDesiredSize:=vector2{X:=FullBakerInfoUI_DepositX},DefaultImage:=Images.EmptyTexture}
                                                                                                    overlay_slot: 
                                                                                                        HorizontalAlignment:=horizontal_alignment.Fill
                                                                                                        VerticalAlignment:=vertical_alignment.Fill
                                                                                                        Widget:=texture_block{DefaultImage:=Images.Img_LineBackground}
                                                                                                    overlay_slot: 
                                                                                                        Padding:=margin{Left:=RowLeftPadding,Top:=RowTopPadding,Right:=RowRightPadding,Bottom:=RowBottomPadding}
                                                                                                        HorizontalAlignment:=horizontal_alignment.Center
                                                                                                        VerticalAlignment:=vertical_alignment.Center
                                                                                                        Widget:=FlourEfficiency_TextBlock
                                            stack_box_slot: #Close
                                                Padding:=margin{Left:=RowLeftPadding,Top:=RowTopPadding,Right:=RowRightPadding,Bottom:=RowBottomPadding}
                                                Widget:=overlay: 
                                                    Slots:=array:
                                                        overlay_slot:
                                                            Widget:=texture_block{DefaultImage:=Images.EmptyTexture,DefaultDesiredSize:=vector2{X:=BigCloseButtonSizeX,Y:=BigCloseButtonSizeY}}
                                                        overlay_slot:
                                                            HorizontalAlignment:=horizontal_alignment.Fill
                                                            VerticalAlignment:=vertical_alignment.Fill 
                                                            Widget:=CloseButton
                                

                                                                                                       
MakeBBInfoUI<constructor>(
    InBakeryMachine : bakery_machines_manager)<transacts> := bb_info_ui :
        BakeryMachinesManager := InBakeryMachine
                                                                                                               