
using { /Fortnite.com/Devices }
using { /Verse.org/Simulation }
using { /UnrealEngine.com/Temporary/Diagnostics }

# See https://dev.epicgames.com/documentation/en-us/uefn/create-your-own-device-in-verse for how to create a verse device.

# A Verse-authored creative device that can be placed in a level
collectibles_manager := class(creative_device):

    var MaybeGameManager : ?game_manager = false

    @editable
    AllCollectibleSecrets : []switch_device = array{}
    @editable
    AllCollectibleHexaChallenges : []switch_device = array{}

    SetGameManager(GameManager : game_manager) : void =
        set MaybeGameManager = option{GameManager}
        SubscribeCollectibles()

    SubscribeCollectibles() : void =
        for(Index:=0..AllCollectibleSecrets.Length-1, CollectibleSecret := AllCollectibleSecrets[Index]):
            CollectibleSecret.TurnedOnEvent.Subscribe(handler_random{MaybeCollectibleManager := option{Self}, Index:=Index}.HandlerCollectibleSwitch)
        for(Index:=0..AllCollectibleHexaChallenges.Length-1, CollectibleSecret := AllCollectibleHexaChallenges[Index]):
            CollectibleSecret.TurnedOnEvent.Subscribe(handler_random{MaybeCollectibleManager := option{Self}, Index:=Index}.HandlerCollectibleHexaSwitch)

    TryRegisterSecret(Agent : agent, Index : int) : void =
        if(GameManager := MaybeGameManager?, Player := player[Agent]<#, PlayerData := GameManager.PlayersDataManager.GetPlayerData[Player]#>):
            #GameManager.PlayersDataManager.SaveNewData(Player, PlayerData, NewData)
            MaybeBakery:=GameManager.GetAgentBakery(Agent)
            Log("Trying to register", LOG_LEVEL_COLLECTIBLES)
            if(Validbakery := MaybeBakery?):
                TupleResultNewData:=GameManager.PlayersDataManager.RegisterSecret(Validbakery.PlayerDataCache,Index)
                NewData:=TupleResultNewData(0)
                WasNewRegister:=TupleResultNewData(1)
                Validbakery.SetNewPlayerDataCache(NewData)
                #GameManager.PlayersDataManager.SaveAllNewData(Player, PlayerData, NewData, Validbakery)
                Validbakery.SetScrollPickaxeEfficiency(NewData)

                var NumSecretsGranted : int = 0
                for(PosibleSecretGranted : NewData.AllSecrets_Granted, PosibleSecretGranted?):
                    set NumSecretsGranted = NumSecretsGranted + 1

                if(WasNewRegister?):
                    GameManager.DisplayCharacterMessage(Agent,W_CHARACTER,"({NumSecretsGranted}/{AllCollectibleSecrets.Length}) Special Data! Useful... for my research. +5% Gather")
                    GameManager.VFXManager.Accolade_BigObjetive.Award(Agent)
                else:
                    GameManager.DisplayCharacterMessage(Agent,W_CHARACTER,"You already got this data. No need to grab it again.")
        else:
            PrintFatalError("On TryRegisterSecret")

    TryRegisterHexaCollectibleSecret(Agent : agent, NumCollectible : int) : void =
        Log("Trying to unlock HexaCollectibleSecret {NumCollectible}",LOG_LEVEL_DEBUG9)
        if(GameManager := MaybeGameManager?):
            MaybeBakery:=GameManager.GetAgentBakery(Agent)
            if:
                Player := player[Agent]
                Validbakery := MaybeBakery?
            then:
                NumChallengesPerCollectible := CalculeNumChallengesPerCollectible(Validbakery.HexaGameManager.HexaChallengesDatabase.AllChallenges.Length,AllCollectibleHexaChallenges.Length)
                #NewLevelsToUnlocked:=NumChallengesPerCollectible+Validbakery.PlayerDataCache.HexaGame_ChallengeLVsUnlocked
                TupleResultNewData:=GameManager.PlayersDataManager.SetTotemHexaChallenge(Validbakery.PlayerDataCache,NumCollectible)
                NewData:=TupleResultNewData(0)
                WasNewRegister:=TupleResultNewData(1)
                #GameManager.PlayersDataManager.SaveNewData(Player, PlayerData, NewData)
                Validbakery.SetNewPlayerDataCache(NewData)
                #GameManager.PlayersDataManager.SaveAllNewData(Player, PlayerData, NewData, Validbakery)
                if(WasNewRegister?):
                    Validbakery.CheckQuestProgress(1.0,UNLOCKHEXAGAME_TUTORIAL_TYPE)
                    GameManager.DisplayCharacterMessage(Agent, W_CHARACTER,"{NumChallengesPerCollectible} new challenges unlocked… Let’s see if you can handle them")
                    GameManager.VFXManager.Accolade_BigObjetive.Award(Agent)
                else:
                    GameManager.DisplayCharacterMessage(Agent, W_CHARACTER,"You’ve already unlocked this. Don’t get ahead of yourself.")
            else:
                PrintFatalError("On TryRegisterSecret")
        else:
            PrintFatalError("On TryRegisterSecret, GameManager not found!")

    InitPlayerCollectibles(Player : player) : void =
        InitCollectibleSecrets(Player)
        InitCollectibleHexaChallenges(Player)

    InitCollectibleHexaChallenges<private>(Player : player) : void =
        if(GameManager := MaybeGameManager?, Player.IsActive[], Agent := agent[Player]):
            MaybeBakery:=GameManager.GetAgentBakery(Agent)
            
            if(ValidBakery:=MaybeBakery?):
                NumChallengesPerCollectible := CalculeNumChallengesPerCollectible(ValidBakery.HexaGameManager.HexaChallengesDatabase.AllChallenges.Length,AllCollectibleHexaChallenges.Length)
                    #var NumTotemsDiscovered : int = 0
                    #for(Index:=0..ValidBakery.PlayerDataCache.AllHexaGame_TotemsUnlocked.Length-1, IsDiscovered:=ValidBakery.PlayerDataCache.AllHexaGame_TotemsUnlocked[Index], IsDiscovered?):
                    #    set NumTotemsDiscovered = NumTotemsDiscovered + 1
                for(Index:=0..ValidBakery.PlayerDataCache.AllHexaGame_TotemsUnlocked.Length-1, IsDiscovered:=ValidBakery.PlayerDataCache.AllHexaGame_TotemsUnlocked[Index], IsDiscovered?, HexaChallengetCollectible:=AllCollectibleSecrets[Index]):
                    HexaChallengetCollectible.TurnOn(Agent)
                    Log("HexaChallenge Collectible {Index} was discovered before, turning on!", LOG_LEVEL_DEBUG9)

    InitCollectibleSecrets<private>(Player : player) : void =
        if(Agent := agent[Player], GameManager := MaybeGameManager?):
            MaybeBakery:=GameManager.GetAgentBakery(Agent)
            if(ValidBakery:=MaybeBakery?):
                if(Player.IsActive[]):
                    AllSecrets_Granted:=ValidBakery.PlayerDataCache.AllSecrets_Granted
                    for(Index:=0..AllSecrets_Granted.Length-1, IsSecretGranted := AllSecrets_Granted[Index], SecretCollectible:=AllCollectibleSecrets[Index]):
                        if(not IsSecretGranted?):
                            Log("Secret Collectible {Index} not granted yet. doing nothing", LOG_LEVEL_DEBUG9)   
                        else:
                            SecretCollectible.TurnOn(Agent)
                            Log("Secret  Collectible {Index} was collected. Checking On it", LOG_LEVEL_DEBUG9)
        

