using { /UnrealEngine.com/Temporary/SpatialMath }

#FAME
MutantMaterialPerCoal : float = 100.0


#Musics

#Languages
ENGLISH : int = 0
SPANISH : int = 1
FRENCH : int = 2
PORTUGUESE : int = 3

#Audio effects
DRONETAKEOFF_SOUND : int = 0
DRONELANDING_SOUND : int = 1
JACKPOT_SOUND : int = 2
SMALLOBJETIVE_SOUND : int = 3
BIGOBJETIVE_SOUND : int = 4
PICKCRITICAL_SOUND : int = 5
DAMAGE_SOUND : int = 6

HEAL_SOUND : int = 7
SHIELD_SOUND : int = 8
TIMEENDING_SOUND : int = 9
PETCREATED_SOUND : int = 10
ERROR_SOUND : int = 11
CHEST_SOUND : int = 12
WIN_SOUND : int = 13
LOSE_SOUND : int = 14
WEAPONCRITICAL_SOUND : int = 15
SUPERSTRIKE_SOUND : int = 16
DUNGEONMUSIC_SOUND : int = 17
DUNGEONMINIBOSSMUSIC_SOUND : int = 18
STARTCHALLENGE_SOUND : int = 19
HEXAGAMEMUSIC_SOUND : int = 20
HIGHVALVE_SOUND : int = 21



IHB_PHASE0_SOUND : int = 22
IHB_PHASE1_SOUND : int = 23
IHB_PHASE2_SOUND : int = 24
IHB_WARNING_SOUND : int = 25

FULL_SOUND : int = 26
MESSAGE_SOUND : int = 27
WARNING_SOUND : int = 28

#Musics
BAKERSMUSIC_SOUND : int = 29
DESERTMUSIC_SOUND : int = 30
SECRETMUSIC_SOUND : int = 31

#Accolades
PICKAXE_ACCOLADE : int = 0
BAKER_ACCOLADE : int = 1
DRONESEND_ACCOLADE : int = 3
HARDBREAD_ACCOLADE : int = 4
INVENTORY_ACCOLADE : int = 5
AGEDSAPPHIRES_ACCOLADE : int = 6

#VALVE STATES
ALLVALVE_STATES_STRINGS : []string = array {"[❌ CLOSSED]","[➡ HALF OPEN]","[➡➡ FULL OPEN]"}

SuperEnergyMultiplier_Shared : float = 0.15
SuperEnergyMultiplier_Owned : float = 2.0

<#
SuperEnergyProductionMultiplier : float = 3.0
SuperEnergySpeedMultiplier : float = 3.0

SuperEnergySpeedMultiplier_Shared : float = 0.25
SuperEnergyProductionMultiplier_Shared : float = 0.25#>

#ALL BAKERY STATS
BakerLevelDegradation : float = 0.998
#MaxUpgradeProductionLevel : int = 150
MaxLoops : int = 5
MinBakeTime : float = 5.0
MaxEfficiency : float = 100.0
BakerBaseEfficiency : float = 5.0
MaxBakerEfficiencyLV : int = 10
    #Deposit
Base_MainWaterDepositCostLimit : float = 10.0
Base_MainFlourDepositCostLimit : float = 10.0
Base_MainCheeseDepositCostLimit : float = 10.0
Base_MainPimentosDepositCostLimit : float = 10.0

    #Backpack
Base_BackpackCostLimit : float = 10.0

#PasiveIncome
Base_PasiveIncome : float = 5.0
Base_PasiveIncomeCost : float = 2.0
MaxUpgradePasiveIncome : int = 20

MaxDepositHeight : float = 1536.0
MinDepositHeight : float = 1024.0
    #Speed
Base_BB_Speed : float = 90.0
Base_CBB_Speed : float = 120.0
Base_PCSB_Speed : float = 180.0
Baker_Speed_Increment : float = 5.0
Baker_Speed_IncrementCost : float = 5.0
    #Efficiency

Base_WaterEfficiency : float = 5.0
Base_FlourEfficiency : float = 4.0
Base_CheeseEfficiency : float = 3.0
Base_PimentosEfficiency : float = 2.0

Baker_Efficiency_IncrementCost : float = 40.0
<#
Base_BB_WaterEfficiency : float = 2.5
Base_CBB_WaterEfficiency : float = 2.5
Base_PCSB_WaterEfficiency : float = 2.5
Base_BB_FlourEfficiency : float = 2.5
Base_CBB_FlourEfficiency : float = 2.5
Base_PCSB_FlourEfficiency : float = 2.5
Base_BB_CheeseEfficiency : float = 2.5
Base_CBB_CheeseEfficiency : float = 2.5
Base_PCSB_CheeseEfficiency : float = 2.5
Base_PCSB_PimentosEfficiency : float = 2.5#>

    #Production
Req_Resourcers_PerSapphire : float = 300.0 #100 #250.0
    #NO MORE THAN 15
Base_B_Production : float = 15.0
Base_CB_Production : float = 15.0
Base_PCS_Production : float = 15.0

#Drones
Base_Drone_Speed : int = 120 #Seven minutes down to 4 maximum 420
Base_Drone_MaxSpeed : int = 20 #Seven minutes down to 4 maximum 240
Base_Drone_UpgradeSpeed : int = 5
Base_B_Storage : float = 25.0
Base_CB_Storage : float = 15.0
Base_PCS_Storage : float = 1000.0
    #Sales
SalesBaseEfficiency : float = 5.0
Sales_Efficiency_IncrementCost : float = 1.0
MaxSalesLV : int = 10

    #Products requirements
    #Bread
Base_Req_WaterFor_B : float = 50.0
Base_Req_FlourFor_B : float = 100.0
    #CheeseBread
Base_Req_WaterFor_CB : float = 150.0
Base_Req_FlourFor_CB : float = 100.0
Base_Req_CheeseFor_CB : float = 50.0
    #PimentoCheeseSandwich
Base_Req_WaterFor_PCS : float = 150.0
Base_Req_FlourFor_PCS : float = 150.0
Base_Req_CheeseFor_PCS : float = 150.0
Base_Req_PimentosFor_PCS : float = 150.0

#Numbers
CHEESEBREADBAKER_MAXTIME_ISSUE : int = 100
CHEESEBREADBAKER_MINTIME_ISSUE : int = 30

PIMENTOCHEESEBLENDER_MAXTIME_ISSUE : int = 60
PIMENTOCHEESEBLENDER_MINTIME_ISSUE : int = 20

BREADBAKER_MAXTIME_ISSUE : int = 120
BREADBAKER_MINTIME_ISSUE : int = 45

#InputIndex
WATER_INPUT_INDEX : int = 0
FLOUR_INPUT_INDEX : int = 1
CHEESE_INPUT_INDEX : int = 2
PIMENTO_INPUT_INDEX : int = 3
BREAD_PROGRESS_INDEX : int = 4
SLICEDBREAD_PROGRESS_INDEX : int = 5
CHEESEBREAD_PROGRESS_INDEX : int = 6
PIMENTOCHEESESANDWICH_PROGRESS_INDEX : int = 7

#Issues
TIMER_INCREMENTED_ISSUES : int = 30
REPAIRED : int = 1
NOT_REPAIRED : int = 0
WATER_OK : float = 100.0
FLOUR_OK : float = 5.0
CHEESE_OK : float = 5.0
PIMENTO_OK : float = 10.0

#Resourcers DO NOT CHANGE THE INDEXES! In the data table they are this way on the backpack
    

    SWORD_TYPE : int = -101
    BOW_TYPE : int = -100

    ARROW : int = -6
    DUNGEONTOKEN : int = -5
    HEXATOKEN : int = -4

    SAPPHIRES : int = -3
    AGEDSAPPHIRES : int = -2
    
    FAME : int = -1
    WATER : int = 0
    FLOUR : int = 1
    CHEESE : int = 2
    PIMENTOS : int = 3
    MUTANTMATERIAL : int = 4
    COAL : int = 5

CHEESE_SUCCESS_SLIDER_MINVALUE : float = 4.28
CHEESE_SUCCESS_SLIDER_MAXVALUE : float = 5.84

ADDED_REPARATION_PROGRESS_WATER : float = 10.0
ADDED_REPARATION_PROGRESS_FLOUR : float = 1.0
ADDED_REPARATION_PROGRESS_CHEESE : int = 10
ADDED_REPARATION_PROGRESS_PIMENTO : int = 10

WATERISSUE_INDEX : int  = 0
FLOURISSUE_INDEX : int  = 1
CHEESEISSUE_INDEX : int  = 2
PIMENTOISSUE_INDEX : int  = 3

B_TYPE : int = 500
CB_TYPE : int = 501
PCS_TYPE : int = 502
SB_TYPE : int = 503


BB_TYPE : int = 0
CBB_TYPE : int = 1
PCSB_TYPE : int = 2
PB_TYPE : int = 3

PIMENTO_SUCCESS_SLIDER_MINVALUE : float = 4.21
PIMENTO_SUCCESS_SLIDER_MAXVALUE : float = 5.81

PIMENTO_ISSUE_SLIDERADDED : int = 30

#Bake aniamtions    
PRODUCT_STARTPOSITION : vector3 = vector3{X:=122.0,Y:=52.0,Z:=152.0}
PRODUCT_MIDPOSITION : vector3 = vector3{X:=0.0,Y:=52.0,Z:=152.0}
PRODUCT_ENDPOSITION : vector3 = vector3{X:=-110.0,Y:=52.0,Z:=152.0}

SB_STARTPOSITION : vector3 = vector3{X:=183.0,Y:=52.0,Z:=56.0}
SB_MIDPOSITION : vector3 = vector3{X:=103.0,Y:=52.0,Z:=56.0}
SB_ENDPOSITION : vector3 = vector3{X:=7.0,Y:=52.0,Z:=56.0}

PCS_STARTPOSITION : vector3 = vector3{X:=0.0,Y:=54.0,Z:=230.5}
PCS_MID1POSITION : vector3 = vector3{X:=0.0,Y:=54.0,Z:=200.0}
PCS_MID2POSITION : vector3 = vector3{X:=0.0,Y:=54.0,Z:=201.5}
PCS_MID3POSITION : vector3 = vector3{X:=0.0,Y:=54.0,Z:=192.0}
PCS_MID4POSITION : vector3 = vector3{X:=0.0,Y:=54.0,Z:=153.5}
PCS_ENDPOSITION : vector3 = vector3{X:=-128.0,Y:=54.0,Z:=153.5}

#Rotation explained:

#Roll is X: Red Axis OF THE OBJECT
#Pitch is Y: Green Axis OF THE OBJECT
#Yaw is Z: Blue axis OF THE OBJECT

#Ship
    #Bread
BSHIP_STARTPOSITION : vector3 = vector3{X:=-214.6,Y:=72.0,Z:=188.0}
BSHIP_MIDPOSITION : vector3 = vector3{X:=-224.0,Y:=64.0,Z:=148.0}
BSHIP_ENDPOSITION : vector3 = vector3{X:=-224.0,Y:=64.0,Z:=16.0}

BSHIP1_RotationRoll : float = 90.0
BSHIP1_RotationPitch : float = 110.0
BSHIP1_RotationYaw : float = 180.0

BSHIP2_RotationRoll : float = 0.0
BSHIP2_RotationPitch : float = 0.0
BSHIP2_RotationYaw : float = 20.0

BSHIP3_RotationRoll : float = 180.0
BSHIP3_RotationPitch : float = 0.0
BSHIP3_RotationYaw : float = 0.0

#Drone
#Roll is X: Red Axis OF THE OBJECT
#Pitch is Y: Green Axis OF THE OBJECT
#Yaw is Z: Blue axis OF THE OBJECT
DroneBaker_Speed_Increment : float = 5.0
DroneBaker_Speed_IncrementCost : float = 25.0

DRONE_Send_RotationRoll : float = -15.0
DRONE_Send_RotationPitch : float = 0.0
DRONE_Send_RotationYaw : float = 0.0

DRONE_Return_RotationRoll : float = 15.0
DRONE_Return_RotationPitch : float = 0.0
DRONE_Return_RotationYaw : float = 0.0


    #Cheese Bread
CBSHIP_STARTPOSITION : vector3 = vector3{X:=-224.0,Y:=64.0,Z:=154.0}
CBSHIP_MIDPOSITION : vector3 = vector3{X:=-224.0,Y:=64.0,Z:=148.0}
CBSHIP_ENDPOSITION : vector3 = vector3{X:=-224.0,Y:=64.0,Z:=24.0}

CBSHIP3_RotationRoll : float = 0.0
CBSHIP3_RotationPitch : float = 0.0
CBSHIP3_RotationYaw : float = 180.0

    #Pimento Cheese Sandwich
PCSSHIP_STARTPOSITION : vector3 = vector3{X:=-264.0,Y:=64.0,Z:=172.0}
PCSSHIP_MIDPOSITION : vector3 = vector3{X:=-264.0,Y:=64.0,Z:=148.0}
PCSSHIP_ENDPOSITION : vector3 = vector3{X:=-264.0,Y:=64.0,Z:=16.0}

PCSSHIP3_RotationRoll : float = 0.0
PCSSHIP3_RotationPitch : float = 0.0
PCSSHIP3_RotationYaw : float = 180.0



StandarSHIP_RotationYaw : float = 0.0
StandarSHIP_RotationPitch : float = 0.0
StandarSHIP_RotationRoll : float = 0.0



using { /UnrealEngine.com/Temporary/SpatialMath }
using { /Verse.org/Colors/NamedColors }
using { /Verse.org/Colors }
using { /Verse.org/Assets }

#MESSAGES
BasicMessage<public><localizes>(InString : string) : message = "{InString}"

#CharacterMessages
MAX_PRIORITY : int = 0
MIN_PRIORITY : int = 99

#UIS
MaxUIDisplayTime : int = 60
BackgroundBlackUISizeX : float = 3840.0
BackgroundBlackUISizeY : float = 2160.0

#Colors
GatherInfoColor : color = Violet

PureHBTextColor : color = Fuchsia
PureHBColor : color = Orchid

ArrowMiniGameTextColor : color = DeepPink
ArrowMiniGameColor : color = Crimson

MaxedColor : color = Gold

SapphiresBackgroundColor : color = DodgerBlue
WaterBackgroundColor : color = Blue
FlourBackgroundColor : color = White
CheeseBackgroundColor : color = Yellow
PimentosBackgroundColor : color = Red

NotEnoughTextColor : color = Black
CurrentTextColor : color = Yellow
UpgradeTextColor : color = GreenYellow
SapphiresTextColor : color = Cyan
HexaTokenTextColor : color = PowderBlue
DungeonTokenTextColor : color = Violet
AgedSapphiresTextColor : color = Firebrick
FameTextColor : color = LightPink
WaterTextColor : color = LightBlue
FlourTextColor : color = GhostWhite
CheeseTextColor : color = LightYellow
PimentoTextColor : color = LightCoral
MobDamageTextColor : color = MediumVioletRed
MutantMaterialTextColor : color = Chartreuse
CoalTextColor : color = LightGray
MutantBreadTextColor : color = PaleGreen

SapphiresColor : color = SteelBlue
AgedSapphiresColor : color = Burlywood
FameColor : color = LightPink
WaterColor : color = Blue
FlourColor : color = DarkGrey
CheeseColor : color = Yellow
PimentoColor : color = Firebrick
MutantMaterialColor : color = ForestGreen
CoalColor : color = DimGray
MutantBreadColor : color = SeaGreen

InfoBakerWaterColor : color = Blue
InfoBakerFlourColor : color = White
InfoBakerCheeseColor : color = Yellow
InfoBakerPimentoColor : color = Red

#ColorTiers
AllTierColors : []color = array{
    Gray,DarkGrey,Silver,White,PaleGreen,
    LightGreen,Chartreuse,SpringGreen,MediumSpringGreen,Aquamarine,Cyan,
    DarkTurquoise,DodgerBlue,SteelBlue,MediumPurple,
    Violet,MediumOrchid,Fuchsia,DeepPink,Red
}

#Material numbers
AllMaterialnumbers : []material = array{Materials.MI_0,Materials.MI_1,Materials.MI_2,Materials.MI_3,Materials.MI_4,
        Materials.MI_5,Materials.MI_6,Materials.MI_7,Materials.MI_8,Materials.MI_9}

#Icon Size
IconSizeX : float = 75.0
IconSizeY : float = 75.0

VerySizeSmallIconX : float = 40.0
VerySizeSmallIconY : float = 40.0

VerySizeSmallIcon : vector2 = vector2{X:=40.0,Y:=40.0}

IconSmallSizeX : float = 50.0
IconSmallSizeY : float = 50.0

ActionIconSizeX : float = 40.0
ActionIconSizeY : float = 40.0

LineSizeX : float = 1000.0
LineSizeY : float = 80.0 #Size of button regular

#Icons margins
TextIcon_LeftMargin : float = 65.0
Icon_LeftMargin : float = 10.0

#Sapphires
SizeBoxSaphiresX : float = 200.0
SizeBoxSaphiresY : float = 80.0

#Image size
HorizontalLineSizeX : float = 5.0
HorizontalLineSizeY : float = 5.0

SizeTextureBlockX : float = 80.0
SizeTextureBlockY : float = 80.0

SizeSmallIconX : float = 50.0
SizeSmallIconY : float = 50.0

#Scroll out
OutOfScreenY : float = 2000.0

#Margins
InterButtonLeftPadding : float = 0.0
InterButtonTopPadding : float = 5.0
InterButtonRightPadding : float = 0.0
InterButtonBottomPadding : float = 5.0

#Margins
InterColumnLeftPadding : float = 5.0
InterColumnTopPadding : float = 0.0
InterColumnRightPadding : float = 5.0
InterColumnBottomPadding : float = 0.0

RowLeftPadding : float = 5.0
RowTopPadding : float = 5.0
RowRightPadding : float = 5.0
RowBottomPadding : float = 5.0

RowLeftHugePadding : float = 10.0
RowTopHugePadding : float = 10.0
RowRightHugePadding : float = 10.0
RowBottomHugePadding : float = 10.0

DefaultTextShadowOpacity : type{_X:float where 0.000000 <= _X, _X <= 1.000000} = 1.0

TextShadowOpacity : float = 0.3
ShadowOffset : ?vector2 = option{vector2{X:=1.0,Y:=1.0}}

#Close Button
BigCloseButtonSizeX : float = 1000.0
BigCloseButtonSizeY : float = 80.0

#Textures
SapphireIcon : texture = Images.Sapphire_Icon
AgedSapphireIcon : texture = Images.AgedSaphire_Icon

PetIcon : texture = Images.Icon_ADC

CheeseIcon : texture = Images.Cheese_Icon
PimentoIcon : texture = Images.Pimento_Icon

OrderIcon : texture = Images.Pipes_Icon
DungeonIcon : texture = Images.Pipes_Icon
HexaGameIcon : texture = Images.Pipes_Icon

BreadIcon : texture = Images.Bread_Icon
CheeseBreadIcon : texture = Images.CheeseBread_Icon
SandwichIcon : texture = Images.PimentoCheeseSandwich_Icon
MutantIcon : texture = Images.Mutant_Material_Icon
CoalIcon : texture = Images.Coal_Icon

ParkourIcon : texture = Images.Coal_Icon


SapphiresTexture : texture = Images.Sapphire_Icon
AgedSapphiresTexture : texture = Images.Sapphire_Icon
FameTexture : texture = Images.Sapphire_Icon
WaterTexture : texture = Images.Water_Icon
FlourTexture : texture = Images.Flour_Icon
CheeseTexture : texture = Images.Cheese_Icon
PimentoTexture : texture = Images.Pimento_Icon
MutantTexture : texture = Images.Mutant_Material_Icon
CoalTexture : texture = Images.Sapphire_Icon

#Resourcers UI
    #Health UIs
ContainerHealth_HealthBarX : float = 740.0
ContainerHealth_HealthBarY : float = 95.0
EntityHealth_MaxHealthBarX : float = 711.3

EntityHealth_TextBottomPadding : float = 2.0

EntityHealth_PaddingSliderLeft : float = 14.3
EntityHealth_PaddingSliderBottom : float = 20.0

MaxActionDisplayTime : int = 7
    #Padding
ResourcersRowLeftPadding : float = 0.0
ResourcersRowTopPadding : float = 2.5
ResourcersRowRightPadding : float = 0.0
ResourcersRowBottomPadding : float = 2.5
    #Proportions
ResourcersX : float = 200.0
ResourcersY : float = 10.0
    #Box Size
#InventoryBoxSizeX : float = 170.0
#InventoryBoxSizeY : float = 50.0

    #Action UI
ActionImgPadding : float = 2.5
XSizeActionUI : float = 200.0
YSizeActionUI : float = 30.0
InventorySliderBarPadding: float = 2.5

    #Resourcers position
XPositionActionUI_Action : float = 150.0
YPositionActionUI_Action : float = -50.0

XPositionUI_WaterRes : float = 150.0
YPositionUI_WaterRes : float = 50.0

XPositionUI_FlourRes : float = 150.0
YPositionUI_FlourRes : float = 110.0

XPositionUI_CheeseRes : float = 150.0
YPositionUI_CheeseRes : float = 170.0

XPositionUI_PimentosRes : float = 150.0
YPositionUI_PimentosRes : float = 230.0

XPositionUI_CoalRes : float = 150.0
YPositionUI_CoalRes : float = 290.0

XPositionUI_MutantRes : float = 150.0
YPositionUI_MutantRes : float = 350.0

#PICKAXE AND GATHER
PickaxeLevelDegradation : float = 0.93
PickaxeTierDegradation : float = 0.74
EXTRA_ProductionTierDegradation : float = 0.88

PercentageToWeakPoint  : float = 0.2


AllPickaxeTiersDegradation : []float = array{
    1.0, # TIER 1
    0.75, # TIER 2
    0.5625, # TIER 3
    0.421875, # TIER 4
    0.31640625, # TIER 5
    0.2373046875, # TIER 6
    0.177978515625, # TIER 7
    0.13348388671875, # TIER 8 #DESERT ZONE
    0.1001129150390625, # TIER 9
    0.07508468627929688, # TIER 10
    0.056313514709472656, # TIER 11
    0.04223513603210449, # TIER 12
    0.03167635202407837, # TIER 13
    0.023757264018058777, # TIER 14
    0.01781794801394408, # TIER 15
    0.01, # TIER 16
    0.009, # TIER 17
    0.007, # TIER 18
    0.005, # TIER 19
    0.004, # TIER 20
    0.003 # TIER 21
}

AllPickaxeTiersDegradation_Example : []float = array{
    1.0, # TIER 1
    0.75, # TIER 2
    0.5625, # TIER 3
    0.421875, # TIER 4
    0.31640625, # TIER 5
    0.2373046875, # TIER 6
    0.177978515625, # TIER 7
    0.13348388671875, # TIER 8
    0.1001129150390625, # TIER 9
    0.07508468627929688, # TIER 10
    0.056313514709472656, # TIER 11
    0.04223513603210449, # TIER 12
    0.03167635202407837, # TIER 13
    0.023757264018058777, # TIER 14
    0.01781794801394408, # TIER 15
    0.013363461010458067, # TIER 16
    0.01002259575784355, # TIER 17
    0.007516946818382662, # TIER 18
    0.005637710113787006, # TIER 19
    0.004228282585340255, # TIER 20
    0.003171211939005191 # TIER 21
}

#HARD BREAD
    HB_MAXLEVEL : int = 19
    #Materials
    Mat_HB_Water : material = Materials.Mat_MasterHardBread_Water
    Mat_HB_WaterWeak : material = Materials.Mat_HB_WaterWeakLow
    Mat_HB_Flour : material = Materials.Mat_MasterHardBread_Flour
    Mat_HB_FlourWeak : material = Materials.Mat_HB_FlourWeakLow
    Mat_HB_Cheese : material = Materials.Mat_MasterHardBread_Cheese
    Mat_HB_CheeseWeak : material = Materials.Mat_HB_CheeseWeakLow
    Mat_HB_Pimentos : material = Materials.Mat_MasterHardBread_Pimentos
    Mat_HB_PimentosWeak : material = Materials.Mat_HB_PimentosWeakLow
    Mat_HB_MutantMaterial : material = Materials.Mat_HB_MutantWeakLow
    Mat_HB_MutantMaterialWeak : material = Materials.Mat_HB_MutantWeakHigh
    Mat_HB_Pure : material = Materials.Mat_HB_Pure
    Mat_HB_PureWeak : material = Materials.Mat_HB_PureWeak
    Mat_HB_MutantMaterialDangerous : material = Materials.Mat_HB_MutantWeakOff
   
    HardBread_Tier_Multiplier : float = 0.05
    #Types
    ENTITY_BOSS : int = 301
    HB_BASETYPE : int = -2 #BASE SLATE BREAD
    HB_TYPEAFK : int = -1 #AFK

    #DONT CHANGE THIS, SHOULD BE THE SAME AS THE MATERIAL MAIN ASSIGNED BECAUSE IS USEN ON ISWEAKPOINT ON GetDamage_And_Resourcers
    HB_TYPE0 : int = 0 #WATER
    HB_TYPE1 : int = 1 #FLOUR
    HB_TYPE2 : int = 2 #CHEESE
    HB_TYPE3 : int = 3 #PIMENTOS

    HB_TYPE4 : int = 4 #MUTANT MATERIAL
    HB_PURE : int = 5 #STEVE PURE HB
    ARROW_MINIGAME : int = 300 #MUTANT MATERIAL
    #HB_TYPE_AFK
    HB_BASETYPE_Water_BaseProb : float = 0.5 #0.5 Posibility
    HB_BASETYPE_Flour_BaseProb : float = 1.0 #0.5 Posibility
    HB_BASETYPE_Cheese_BaseProb : float = 0.0 #0.0 Posibility
    HB_BASETYPE_Pimentos_BaseProb : float = 0.0 #0.0 Posibility
    #HB_TYPE_AFK
    HB_TYPEAFK_Water_BaseProb : float = 0.25 #0.70 Posibility
    HB_TYPEAFK_Flour_BaseProb : float = 0.5 #0.1 Posibility
    HB_TYPEAFK_Cheese_BaseProb : float = 0.75 #0.1 Posibility
    HB_TYPEAFK_Pimentos_BaseProb : float = 1.0 #0.1 Posibility
    #TYPE 0 - WATER
    HB_TYPE0_Water_BaseProb : float = 0.7 #0.70 Posibility
    HB_TYPE0_Flour_BaseProb : float = 0.8 #0.1 Posibility
    HB_TYPE0_Cheese_BaseProb : float = 0.9 #0.1 Posibility
    HB_TYPE0_Pimentos_BaseProb : float = 1.0 #0.1 Posibility
    #HB_TYPE0_MutantMaterial_BaseProb : float = 1.00 #0.03 Posibility
    #TYPE 1 - FLOUR
    HB_TYPE1_Water_BaseProb : float = 0.1 #0.09 Posibility
    HB_TYPE1_Flour_BaseProb : float = 0.8 #0.70 Posibility
    HB_TYPE1_Cheese_BaseProb : float = 0.9 #0.09 Posibility
    HB_TYPE1_Pimentos_BaseProb : float = 1.0 #0.09 Posibility
    #HB_TYPE1_MutantMaterial_BaseProb : float = 1.00 #0.03 Posibility
    #TYPE 2 - CHEESE
    HB_TYPE2_Water_BaseProb : float = 0.1 #0.09 Posibility
    HB_TYPE2_Flour_BaseProb : float = 0.2 #0.09 Posibility
    HB_TYPE2_Cheese_BaseProb : float = 0.9 #0.70 Posibility
    HB_TYPE2_Pimentos_BaseProb : float = 1.0 #0.09 Posibility
    #HB_TYPE2_MutantMaterial_BaseProb : float = 1.00 #0.03 Posibility
    #TYPE 3 - PIMENTOS
    HB_TYPE3_Water_BaseProb : float = 0.1 #0.09 Posibility
    HB_TYPE3_Flour_BaseProb : float = 0.2 #0.09 Posibility
    HB_TYPE3_Cheese_BaseProb : float = 0.3 #0.70 Posibility
    HB_TYPE3_Pimentos_BaseProb : float = 1.0 #0.09 Posibility
    #HB_TYPE3_MutantMaterial_BaseProb : float = 1.00 #0.03 Posibility
    #TYPE 4 - MUTANT
    HB_TYPE4_Water_BaseProb : float = 0.25 #0.175 Posibility
    HB_TYPE4_Flour_BaseProb : float = 0.50 #0.175 Posibility
    HB_TYPE4_Cheese_BaseProb : float = 0.75 #0.175 Posibility
    HB_TYPE4_Pimentos_BaseProb : float = 1.00 #0.175 Posibility
    #HB_TYPE4_MutantMaterial_BaseProb : float = 1.00 #0.30 Posibility
    
#PICKAXE
    
    #INDEXES
    CRITICAL_PROB : int = 0
    CRITICAL_DAMAGE : int = 1
    PICKAXE_LUCK : int = 2
    PICKAXE_WEAK_MULT : int = 3
    PICKAXE_MULT : int = 4
#Strength
    Base_PickaxeStrength : float = 15.0
    #Max level
    MaxTierLevel : int = 25 #150 From 31 is negative!
    #Efficiency
    PickaxeBaseScrollEfficiency : float = 0.05
    PickaxeBaseEfficiencyCost : float = 14.0
    PickaxeBaseEfficiency : float = 5.0
    PickaxeMaxEfficiencyLV : int = 20
    #PickaxeMaxEfficiency : float = 100.0
    
    #Efficiency
    PickaxeBaseProbCritical : float = 5.0
    #Base
    PickaxeBaseStrenght : float = 15.0
    #Multiple res Gather
    MultipleResourceGatherCostBase: float = 100000000.0
    MultipleResourceGather : float = 0.25
    #Multiplier
    WeakPointMultBase : float = 0.5
    WeakPointCostBase: float = 20.0
    MaxWeakPointMult : float = 7.0
    #Critical
    MaxPickaxeCriticalProb : float = 1.0
    MaxPickaxeCriticalDamage : float = 3.0 #It is +1 afterwards

    PickaxeCriticalProbBase : float = 0.05
    PickaxeCriticalDamageBase : float = 0.15

    PickaxeCriticalProbCostBase : float = 10.0
    PickaxeCriticalDamageCostBase : float = 25.0
    
   
    #Luck
    CoalProbabilityBase : float = 0.01
    MaxCoalProbability : float = 0.1
    HBLevelPer_Coal : float = 5.0
    #MultiResource
    MaxMultiResource : float = 2.0

#INVENTORY
    BackpackBase_Water : float = 3000.0
    BackpackBase_WaterLimit : float = 1000000000000000000000000000000.0
    BackpackBase_Flour : float = 3000.0
    BackpackBase_CheeseLimit : float = 1000000000000000000000000000000.0
    BackpackBase_Cheese : float = 2000.0
    BackpackBase_Pimentos : float = 1000.0
    BackpackBase_PimentoLimit : float = 1000000000000000000000000000000.0
    BackpackBase_MutantMaterial : float = 500.0
    BackpackBase_MutantMaterialLimit : float = 1000000000000.0

#Info baker

FullBakerInfoUI_DepositX : float = 300.0
FullBakerInfoUI_DepositY : float = 420.0

FullProgressBakerInfoUI_X : float = 1250.0


#Pets
###############

MaxPurity:int=99
MaxTier:int=3

RecycleEfficiency:float=0.90
FusionEfficiency:float=0.99

MaxLV_Pet_PackageA:int=20
MaxLV_Pet_PackageB:int=20

AllStarsMaterials : []material = array{Materials.MI_Star_1,Materials.MI_Star_2,Materials.MI_Star_3}


EnemyPetBarMaterial : material = Materials.MI_Life_Red
FriendlyPetBarMaterial : material = Materials.MI_Life_Green
GreyPetBarMaterial : material = Materials.MI_Life_Grey
DarkGreenPetBarMaterial : material = Materials.MI_Life_DarkGreen
DarkRedPetBarMaterial : material = Materials.MI_Life_DarkRed
YellowPetBarMaterial : material = Materials.MI_Life_Yellow
BluePetBarMaterial : material = Materials.MI_Life_Blue

#Pets Interface
SelectorBoxPadding : float = 2.5
SuperLowPadding : float = 2.0
SelectorBoxLinePadding : float = 5.0

MaxHealthSize_PetHexaUI : float = 375.0

SizeSlotX_PetHexaUI : float = 380.0
SizeSlotY_PetHexaUI : float = 130.0

SizeImageX_PetHexaUI : float = 100.0
SizeImageY_PetHexaUI : float = 100.0

PetsLineBoxPadding : float = 2.5
TextPetsBoxPadding : float = 60.0

PetsSelectorImageSize : vector2 = vector2{X:=140.0,Y:=140.0}

QuietButton_PetSelectorDesiredSize : vector2 = vector2{X:=150.0,Y:=56.0}

    #Pet type

PetPackageA_BaseCost : float = 10.0
PetPackageB_BaseCost : float = 10.0
#PetPurity_BaseCost : float = 5.0
#PetTier_BaseCost : float = 7.0

PetPlayerCorrection : vector3 = vector3{X:=0.0,Y:=0.0,Z:=-84.0}
PetPlayerCorrectionCrouching : vector3 = vector3{X:=0.0,Y:=0.0,Z:=-58.0}

PetScale_Tier1 : vector3 = vector3{X:=0.6,Y:=0.6,Z:=0.6}
PetPivotVariation_Tier1 : vector3 = vector3{X:=0.0,Y:=0.0,Z:=39.0}

PetScale_Tier2 : vector3 = vector3{X:=0.8,Y:=0.8,Z:=0.8}
PetPivotVariation_Tier2 : vector3 = vector3{X:=0.0,Y:=0.0,Z:=50.0}

PetScale_Tier3 : vector3 = vector3{X:=1.0,Y:=1.0,Z:=1.0}
PetPivotVariation_Tier3 : vector3 = vector3{X:=0.0,Y:=0.0,Z:=63.0}

AI_PET : logic = false
PLAYER_PET : logic = true

EMPTY : int = -1


ALL_PLAYER_PICKDAMAGE_MULTIPLIER : float = 0.01
ALL_PLAYER_WEAKPOINT_MULTIPLIER : float = 0.01

ALL_PLAYER_PICKEFFICIENCY_MULTIPLIER : float = 0.01
ALL_PLAYER_PICKCRITICALDAMAGE_MULTIPLIER : float = 0.01
    #Player enhacements
ADC : int = 0
ADC_PLAYER_SPECIALA_BOWMULTIPLIER : float = 0.01
ADC_PLAYER_SPECIALB_BASETIME : float = 200.0
ADC_PLAYER_SPECIALB_IMPROVETIME : float = 0.25

DRUID : int = 1
DRUID_PLAYER_SPECIALA_HEAL : float = 0.01
DRUID_PLAYER_SPECIALB_EMERGENCY : float = 0.5

TANK : int = 2
TANK_PLAYER_SPECIALA_HEALTH : float = 0.5
TANK_PLAYER_SPECIALB_SHIELD : float = 0.2

FIGHTER : int = 3
FIGHTER_PLAYER_SPECIALA_SWORDDAMAGE : float = 0.01
FIGHTER_PLAYER_SPECIALB_ULTRASTRIKE : float = 0.02

HUNTER : int = 4
HUNTER_PLAYER_SPECIALA_HUNTPROB : float = 0.01
HUNTER_PLAYER_SPECIALB_WEAPONDAMAGE : float = 0.01


    #Tournament stats
ADC_HEALTH_MULTIPLIER : float = 0.35
ADC_DAMAGE_MULTIPLIER : float = 0.70
ADC_HEALING_MULTIPLIER : float = 0.0
ADC_SPEED_MULTIPLIER : float = 0.8
ADC_ATTACKSPEED_MULTIPLIER : float = 1.25

DRUID_HEALTH_MULTIPLIER : float = 0.35
DRUID_DAMAGE_MULTIPLIER : float = 0.35
DRUID_HEALING_MULTIPLIER : float = 1.45
DRUID_SPEED_MULTIPLIER : float = 0.85
DRUID_ATTACKSPEED_MULTIPLIER : float = 1.1

TANK_HEALTH_MULTIPLIER : float = 2.75
TANK_DAMAGE_MULTIPLIER : float = 0.75
TANK_HEALING_MULTIPLIER : float = 0.0
TANK_SPEED_MULTIPLIER : float = 0.8
TANK_ATTACKSPEED_MULTIPLIER : float = 0.9

FIGHTER_HEALTH_MULTIPLIER : float = 1.25
FIGHTER_DAMAGE_MULTIPLIER : float = 1.2
FIGHTER_HEALING_MULTIPLIER : float = 0.0
FIGHTER_SPEED_MULTIPLIER : float = 1.2
FIGHTER_ATTACKSPEED_MULTIPLIER : float = 1.1

HUNTER_HEALTH_MULTIPLIER : float = 0.85
HUNTER_DAMAGE_MULTIPLIER : float = 1.25
HUNTER_HEALING_MULTIPLIER : float = 0.0
HUNTER_SPEED_MULTIPLIER : float = 1.2
HUNTER_ATTACKSPEED_MULTIPLIER : float = 1.15

#Pet stats types


STAT_TYPE_HEALTH : int = 0
STAT_TYPE_DAMAGE : int = 1
STAT_TYPE_SPEED : int = 2

STAT_TYPE_PLAYER_SPECIAL_A : int = 3
STAT_TYPE_PLAYER_PICKDAMAGE : int = 4
STAT_TYPE_PLAYER_WEAKPOINTMULT : int = 5

STAT_TYPE_ATTACKSPEED : int = 6
STAT_TYPE_HEALING : int = 7
STAT_TYPE_RANGE : int = 8

STAT_TYPE_PLAYER_SPECIAL_B : int = 9
STAT_TYPE_PLAYER_PICKEFFICENCY : int = 10
STAT_TYPE_PLAYER_PICKCRITICALMULT : int = 11

STAT_TYPE_PET_PERK_MAXACTIVE : int = 12
STAT_TYPE_PET_PERK_MAXSLOTS : int = 13
STAT_TYPE_PET_PERK_TIER : int = 14

    #Pet Perks
ActivePetSlotsCostBase : float = 5.0
PetSlotsCostBase : float = 3.0
PetTierCostBase : float = 5.0

MaxActivePetSlots : int = 3
MaxPetSlots : int = 20
MaxPetTier : int = 3

    #PetBases
NeutralScale : vector3 = vector3{X:=1.0,Y:=1.0,Z:=1.0}

HexaGameProjectileSpeed:float=450.0
HexaGameHealProjectileSpeed:float=300.0
HexaGameProjectileElevation:float=64.0


MeleeTilesRange : float = 1.0
LongRangeTilesRange : float = 2.0
SuperRangeTilesRange : float = 3.0
UltraRangeTilesRange : float = 3.0

MeleeAttackRange : float = 156.0
LongRangeAttackRange : float = 312.0
SuperRangeAttackRange : float = 468.0
UltraRangeAttackRange : float = 468.0

PetBase_Damage:float=25.0

PetBasePurity:int=1
PetBasePurityCost:float=5.0
PetMaxPurity:int=99
PetTotalMaxPurity:int=299
ReqDestroyedHB_PerPurity:int=75

PetBase_Health:float=100.0
PetBase_HealingPerk:float=15.0

PetBase_AttackSpeed:float=2.5
PetBase_AttackSpeedImprovement:float=0.07
PetBase_AttackSpeedCost:float=5000000.0

PetBase_Speed:float=3.0
PetBase_SpeedImprovement:float=0.1
PetBase_SpeedCost:float=5000000.0

PetBase_Range:int=1

#Pet UI
SUBMENU_CREATE_PET : int = 0
SUBMENU_IMPROVE_PET : int = 1
SUBMENU_BUY_SLOTSPET : int = 2
SUBMENU_MIX_PET : int = 3
SUBMENU_EQUIP_PET : int = 4
SUBMENU_RECYCLE_PET : int = 5

From_EquipPetsUI : int = 0
From_MixPetsUI : int = 1

    #Base damage
SwordBaseDamage : float = 15.0
BowBaseDamage : float = 20.0
    #Damage cost
SwordBaseDamageCost : float = 5.0
BowBaseDamageCost : float = 6.0
    #Critical prob
SwordBaseCriticalProb : float = 0.1
BowBaseCriticalProb : float = 0.05
    #Critical prob cost
SwordBaseCriticalProbCost : float = 10.0
BowBaseCriticalProbCost : float = 15.0
    #Critical
SwordBaseCritical : float = 0.1
BowBaseCritical : float = 0.1
    #Critical prob cost
SwordBaseCriticalCost : float = 8.0
BowBaseCriticalCost : float = 9.0
    #Mutation farmer
SwordBaseReward : float = 0.1
BowBaseReward: float = 0.7
    #Mutation farmer cost
SwordBaseRewardCost : float = 7.0
BowBaseRewardCost : float = 5.0

MaxSuperStrikeMultiplier : float = 5.0

WEAPON_STATMAXLEVEL : int = 20
WEAPONSTAT_TYPE_EVOLUTION : int = 0
WEAPONSTAT_TYPE_DAMAGE : int = 1
WEAPONSTAT_TYPE_CRITICALPROB : int = 2
WEAPONSTAT_TYPE_CRITICALDAMAGE : int = 3
WEAPONSTAT_TYPE_REWARD : int = 4


NUMHITS_FOR_SUPERSTRIKE : float = 7.0


MOBSTARTINGHEALTH : float = 1000.0

Reward_NormalMobMultiplier : float = 1.0
Reward_DungeonMobMultiplier : float = 0.25

Reward_MiniBossMultiplier : float = 5.0
Reward_GroundMobMultiplier : float = 1.0
Reward_FlyingMobMultiplier : float = 0.25

MAX_NPCMOBSPAWNER_TIMEUSE : int = 240 #4 Minutes

#MOBS
RandomMobMaxLifeTime : float = 30.0

RandomPlantEncounter_PosibilityBase : float = 0.05
RandomPlantEncounter_Posibility_PerHitAdded : float = 0.003
RandomPlantEncounter_MaxBasePosibility : float = 0.1

    #Flying mob
#PlantMobBaseHealth : float = 1.5
PlantMobBaseNumHits : float = 4.0

    #Ground mob
#GroundMobBaseHealth : float = 1.0
GroundMobBaseNumHits : float = 3.0

    #Druid mob
#DruidMobBaseHealth : float = 0.75
DruidMobBaseNumHits : float = 2.0

    #Miniboss
#MINIBOSS_CHECK_RANGE : float = 1500.0

MiniBossMultiplier : float = 25.0
#BossMultiplier : float = 25.0

DRUID_MOB : int = 0
GROUND_MOB : int = 1
PLANT_MOB : int = 2
MINIBOSS_MOB : int = 3
#MINIBOSS_TYPE : int = 4

#MOB_TYPE : int = 5
#MOBDUNGEON_TYPE : int = 6

SOUND_BITE : int = 0
SOUND_DIRT : int = 1
SOUND_DEATH : int = 2
SOUND_LOAD_ZONE : int = 3
SOUND_DAMAGE_ZONE : int = 4
SOUND_BOSS_LAUGH : int = 5
SOUND_HEAL_ZONE : int = 6

#DUNGEONS
DungeonEvolution_Base : float = 0.05

PlayerBaseHealth : float = 60.0

PetPlayerEnhacement_HealDelay : float = 30.0
PetPlayerEnhacement_ShieldDelay  : float = 20.0

ENTER_TYPE_BUTTON : int = 0
EXIT_TYPE_BUTTON : int = 1
TREASURE_TYPE_BUTTON : int = 2
OTHER_INTERACTION_TYPE_BUTTON : int = 3

#AWARD TYPES 
#MAINCHEST_TYPE : int = 0
#SECONDARYCHEST_TYPE : int = 1
#Roll is X: Red Axis OF THE OBJECT
#Pitch is Y: Green Axis OF THE OBJECT
#Yaw is Z: Blue axis OF THE OBJECT

#HEXA AND DUNGEON PERKS
HexaAndDungeons_MaxCoal : int = 75

    #HexaGame
HEXAGAME_FirstTimeMultiplier : int = 2
HEXAGAME_OFFICIAL_NAME : string = "Sourdough Bread Competition" 

HEXAGAMEPERK_DELAY : int = 0
HEXAGAMEPERK_ACCUMULATED : int = 1

HexaGame_DelayTime_Base : float = 240.0 
HexaGame_DelayTime_MinTime : float = 120.0 
HexaGame_DelayTime_ImprovementBase : float = 12.0

HexaGamePerk_DelayBaseCost : float = 50.0
HexaGamePerk_AccumulatedBaseCost : float = 1.0

HexaGame_MaxToken_Accumulated : int = 21
HexaGame_MinToken_Accumulated_MinusOne : int = 4

    #Dungeon Perk
DUNGEONPERK_DELAY : int = 0
DUNGEONPERK_ACCUMULATED : int = 1

Dungeon_DelayTime_Base : float = 720.0 
Dungeon_DelayTime_MinTime : float = 360.0 
Dungeon_DelayTime_ImprovementBase : float = 36.0

DungeonPerk_DelayBaseCost : float = 500.0
DungeonPerk_AccumulatedBaseCost : float = 5.0

Dungeon_MaxToken_Accumulated : int = 5


    
    #Tutorials
USETELEPORTER_TUTORIAL_TYPE : int = 800
USEFUNNEL_TUTORIAL_TYPE : int = 801
USEVALVES_TUTORIAL_TYPE : int = 802
UNLOCKAREA_TUTORIAL_TYPE : int = 803
IMPROVEPICKAXE_TUTORIAL_TYPE : int = 804
IMPROVEBACKPACK_TUTORIAL_TYPE : int = 805
IMPROVEMAINDEPOSITS_TUTORIAL_TYPE : int = 806
CHECKINFO_TUTORIAL_TYPE : int = 807
IMPROVEBAKERY_TUTORIAL_TYPE : int = 808
IMPROVEPICKPERK_TUTORIAL_TYPE : int = 809
BREAKBREAD_TUTORIAL_TYPE : int = 810
CLOSINGVALVES_TUTORIAL_TYPE : int = 811
OPENINGVALVES_TUTORIAL_TYPE : int = 812
HALFVALVES_TUTORIAL_TYPE : int = 813
USECOAL_TUTORIAL_TYPE : int = 814
USEAFKZONE_TUTORIAL_TYPE : int = 815
CREATEPET_TUTORIAL_TYPE : int = 816
IMPROVEPET_TUTORIAL_TYPE : int = 817
USEEXPENSIVETELEPORTER_TUTORIAL_TYPE : int = 818
UNLOCKHEXAGAME_TUTORIAL_TYPE : int = 819
WINHEXAGAME_TUTORIAL_TYPE : int = 820
COMPLETEDUNGEON_TUTORIAL_TYPE : int = 821
WEAPONIMPROVE_TUTORIAL_TYPE : int = 822
TOKENIMPROVE_TUTORIAL_TYPE : int = 823
AGGEDSAPPHIRESUSED_TUTORIAL_TYPE : int = 824
PCSBOPENINGVALVES_TUTORIAL_TYPE : int = 825
BAKERTUTORIAL_TUTORIAL_TYPE : int = 826
PETTUTORIAL_TUTORIAL_TYPE : int = 827
SENDDRONE_TUTORIAL_TYPE : int = 828
PICKHIT_TUTORIAL_TYPE : int = 829
COMPLETEBAKE_TUTORIAL_TYPE : int = 830
LANGUAGE_TUTORIAL_TYPE : int = 831

    #Quests
FINDSCROLL_AREA0_QUEST_TYPE : int = 201
FINDSCROLL_AREA2_QUEST_TYPE : int = 202
CHANGEVALVES_QUEST_TYPE : int = 204
ELECTRICALSYSTEM_MINIGAME_QUEST_TYPE : int = 205
TRUCKENGINE_MINIGAME_QUEST_TYPE : int = 206
PREASSURESYSTEM_MINIGAME_QUEST_TYPE : int = 207
ARCHERY_MINIGAME_QUEST_TYPE : int = 208
SIMONCOLOR_MINIGAME_QUEST_TYPE : int = 209
SIMONSOUND_MINIGAME_QUEST_TYPE : int = 210
UNSTABLEHB_QUEST_TYPE : int = 211
STEVEBAKER_QUEST_TYPE : int = 212
FINDSCROLL_AREA8_QUEST_TYPE : int = 213
MASTERY_QUEST_TYPE : int = 214

Quest_DelayTime_Base : float = 900.0 #900.0

    #Order
Order_DelayTime_Base : float = 700.0 #600.0 700.0 until bug is solved

#COAL MULTIPLIER
Coal_To_SuperEnergy : int = 3
MaxTime_SuperEnergy : float = 3600.0 #1 Hour

#Teleporters
TELEPORTER_FROM_BAKERY_AREA : int = 0
TELEPORTER_FROM_LOBBY_AREA : int = 1
TELEPORTER_FROM_HEXAGAME_AREA : int = 2


#Characters
W_CHARACTER : int = 0
H_CHARACTER : int = 1
G_CHARACTER : int = 2
J_CHARACTER : int = 3
I_CHARACTER : int = 4

#Special event
SPECIALEVENT_FIRSTENCOUNTER : int = 0
SPECIALEVENT_MAINLORE : int = 1

#Inestability
INSTABLE_BREAD_MULTIPLIER : float = 3.0

#Mastery
MinExponential_ForAgedSapphires : int = 9 #GIGAS
Exponetial_ForAgedSapphires : int = 1 #Every Added 0
AgedSapphiresPer_Exponential : int = 5 #Every jump is 5 Sapphires


