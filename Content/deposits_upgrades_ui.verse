using { /UnrealEngine.com/Temporary/UI }
using { /UnrealEngine.com/Temporary/SpatialMath }
using { /Verse.org/Simulation }
using { /Fortnite.com/Devices }
using { /Fortnite.com/UI }
using { /Verse.org/Colors/NamedColors }
using { /Verse.org/Colors }

# See https://dev.epicgames.com/documentation/en-us/uefn/create-your-own-device-in-verse for how to create a verse device.

# A Verse-authored creative device that can be placed in a level
deposits_upgrades_ui := class():

    #Class settings
    var MaybeAgent : ?agent = false

    BakeryMachinesManager : bakery_machines_manager
    
    BackgroundImageSizeX : float = 1700.0
    BackgroundImageSizeY : float = 1100.0

    #LineSizeX : float = 1000.0
    #LineSizeY : float = 80.0 #Size of button regular

    RColumn1SizeX : float = 80.0
    RColumn1SizeY : float = 80.0
    RColumn2SizeX : float = 175.0
    RColumn2SizeY : float = 80.0
    RColumn3SizeX : float = 200.0
    RColumn3SizeY : float = 80.0
    RColumn4SizeX : float = 200.0
    RColumn4SizeY : float = 80.0
    RColumn5SizeX : float = 225.0
    RColumn5SizeY : float = 80.0
    RColumn6SizeX : float = 250.0
    RColumn6SizeY : float = 80.0
    RColumn7SizeX : float = 200.0
    RColumn7SizeY : float = 80.0
    RColumn8SizeX : float = 225.0
    RColumn8SizeY : float = 80.0

    TitleText:text_block=text_block{DefaultTextColor:=White, DefaultShadowOffset:=ShadowOffset, DefaultShadowOpacity := DefaultTextShadowOpacity}

    #Current values
    TotalWater_TextBlock : text_block = text_block{DefaultTextColor:=WaterTextColor, DefaultShadowOffset:=ShadowOffset, DefaultShadowOpacity := DefaultTextShadowOpacity}
    TotalFlour_TextBlock : text_block = text_block{DefaultTextColor:=FlourTextColor, DefaultShadowOffset:=ShadowOffset, DefaultShadowOpacity := DefaultTextShadowOpacity}
    TotalCheese_TextBlock : text_block = text_block{DefaultTextColor:=CheeseTextColor, DefaultShadowOffset:=ShadowOffset, DefaultShadowOpacity := DefaultTextShadowOpacity}
    TotalPimentos_TextBlock : text_block = text_block{DefaultTextColor:=PimentoTextColor, DefaultShadowOffset:=ShadowOffset, DefaultShadowOpacity := DefaultTextShadowOpacity}

    TotalSapphires_BackgroundTexture : texture_block = texture_block{DefaultTint:=AgedSapphiresTextColor, DefaultDesiredSize:=vector2{X:=SizeBoxSaphiresX,Y:=SizeBoxSaphiresY}, DefaultImage:=Images.Img_LineBackground2}
    TotalAgedSapphires_BackgroundTexture : texture_block = texture_block{DefaultTint:=AgedSapphiresTextColor, DefaultDesiredSize:=vector2{X:=SizeBoxSaphiresX,Y:=SizeBoxSaphiresY}, DefaultImage:=Images.Img_LineBackground2}

    TotalWater_BackgroundTexture : texture_block = texture_block{DefaultTint:=WaterBackgroundColor, DefaultDesiredSize:=vector2{X:=SizeBoxSaphiresX,Y:=SizeBoxSaphiresY}, DefaultImage:=Images.Img_LineBackground2}
    TotalFlour_BackgroundTexture : texture_block = texture_block{DefaultTint:=FlourBackgroundColor, DefaultDesiredSize:=vector2{X:=SizeBoxSaphiresX,Y:=SizeBoxSaphiresY}, DefaultImage:=Images.Img_LineBackground2}
    TotalCheese_BackgroundTexture : texture_block = texture_block{DefaultTint:=CheeseBackgroundColor, DefaultDesiredSize:=vector2{X:=SizeBoxSaphiresX,Y:=SizeBoxSaphiresY}, DefaultImage:=Images.Img_LineBackground2}
    TotalPimentos_BackgroundTexture : texture_block = texture_block{DefaultTint:=PimentosBackgroundColor, DefaultDesiredSize:=vector2{X:=SizeBoxSaphiresX,Y:=SizeBoxSaphiresY}, DefaultImage:=Images.Img_LineBackground2}

    TotalSapphires_TextBlock : text_block = text_block{DefaultTextColor:=AgedSapphiresTextColor, DefaultShadowOffset:=ShadowOffset, DefaultShadowOpacity := DefaultTextShadowOpacity}
    TotalAgedSapphires_TextBlock : text_block = text_block{DefaultTextColor:=AgedSapphiresTextColor, DefaultShadowOffset:=ShadowOffset, DefaultShadowOpacity := DefaultTextShadowOpacity}

    #Resourcer deposit
    R_Row1_Water_TextureBlock : texture_block = texture_block{DefaultImage:=Images.Water_Icon, DefaultDesiredSize:= vector2{X:=IconSmallSizeX,Y:=IconSmallSizeY}}
    R_Row2_Flour_TextureBlock : texture_block = texture_block{DefaultImage:=Images.Flour_Icon, DefaultDesiredSize:= vector2{X:=IconSmallSizeX,Y:=IconSmallSizeY}}
    R_Row3_Cheese_TextureBlock : texture_block = texture_block{DefaultImage:=Images.Cheese_Icon, DefaultDesiredSize:= vector2{X:=IconSmallSizeX,Y:=IconSmallSizeY}}
    R_Row4_Pimentos_TextureBlock : texture_block = texture_block{DefaultImage:=Images.Pimento_Icon, DefaultDesiredSize:= vector2{X:=IconSmallSizeX,Y:=IconSmallSizeY}}

    R_Row1_Capacity_TextBlock : text_block = text_block{DefaultTextColor:=CurrentTextColor, DefaultShadowOffset:=ShadowOffset, DefaultShadowOpacity := DefaultTextShadowOpacity}
    R_Row2_Capacity_TextBlock : text_block = text_block{DefaultTextColor:=CurrentTextColor, DefaultShadowOffset:=ShadowOffset, DefaultShadowOpacity := DefaultTextShadowOpacity}
    R_Row3_Capacity_TextBlock : text_block = text_block{DefaultTextColor:=CurrentTextColor, DefaultShadowOffset:=ShadowOffset, DefaultShadowOpacity := DefaultTextShadowOpacity}
    R_Row4_Capacity_TextBlock : text_block = text_block{DefaultTextColor:=CurrentTextColor, DefaultShadowOffset:=ShadowOffset, DefaultShadowOpacity := DefaultTextShadowOpacity}

    R_Row1_CapacityUpgrade_TextBlock : text_block = text_block{DefaultTextColor:=UpgradeTextColor, DefaultShadowOffset:=ShadowOffset, DefaultShadowOpacity := DefaultTextShadowOpacity}
    R_Row2_CapacityUpgrade_TextBlock : text_block = text_block{DefaultTextColor:=UpgradeTextColor, DefaultShadowOffset:=ShadowOffset, DefaultShadowOpacity := DefaultTextShadowOpacity}
    R_Row3_CapacityUpgrade_TextBlock : text_block = text_block{DefaultTextColor:=UpgradeTextColor, DefaultShadowOffset:=ShadowOffset, DefaultShadowOpacity := DefaultTextShadowOpacity}
    R_Row4_CapacityUpgrade_TextBlock : text_block = text_block{DefaultTextColor:=UpgradeTextColor, DefaultShadowOffset:=ShadowOffset, DefaultShadowOpacity := DefaultTextShadowOpacity}

    R_Row1_CapacityUpgradeCost_TextBlock : text_block = text_block{DefaultTextColor:=SapphiresTextColor, DefaultShadowOffset:=ShadowOffset, DefaultShadowOpacity := DefaultTextShadowOpacity}
    R_Row2_CapacityUpgradeCost_TextBlock : text_block = text_block{DefaultTextColor:=SapphiresTextColor, DefaultShadowOffset:=ShadowOffset, DefaultShadowOpacity := DefaultTextShadowOpacity}
    R_Row3_CapacityUpgradeCost_TextBlock : text_block = text_block{DefaultTextColor:=SapphiresTextColor, DefaultShadowOffset:=ShadowOffset, DefaultShadowOpacity := DefaultTextShadowOpacity}
    R_Row4_CapacityUpgradeCost_TextBlock : text_block = text_block{DefaultTextColor:=SapphiresTextColor, DefaultShadowOffset:=ShadowOffset, DefaultShadowOpacity := DefaultTextShadowOpacity}

    #Pasive income
    P_Row1_Pasive_TextBlock : text_block = text_block{DefaultTextColor:=CurrentTextColor, DefaultShadowOffset:=ShadowOffset, DefaultShadowOpacity := DefaultTextShadowOpacity}
    P_Row2_Pasive_TextBlock : text_block = text_block{DefaultTextColor:=CurrentTextColor, DefaultShadowOffset:=ShadowOffset, DefaultShadowOpacity := DefaultTextShadowOpacity}
    P_Row3_Pasive_TextBlock : text_block = text_block{DefaultTextColor:=CurrentTextColor, DefaultShadowOffset:=ShadowOffset, DefaultShadowOpacity := DefaultTextShadowOpacity}
    P_Row4_Pasive_TextBlock : text_block = text_block{DefaultTextColor:=CurrentTextColor, DefaultShadowOffset:=ShadowOffset, DefaultShadowOpacity := DefaultTextShadowOpacity}

    P_Row1_PasiveUpgrade_TextBlock : text_block = text_block{DefaultTextColor:=UpgradeTextColor, DefaultShadowOffset:=ShadowOffset, DefaultShadowOpacity := DefaultTextShadowOpacity}
    P_Row2_PasiveUpgrade_TextBlock : text_block = text_block{DefaultTextColor:=UpgradeTextColor, DefaultShadowOffset:=ShadowOffset, DefaultShadowOpacity := DefaultTextShadowOpacity}
    P_Row3_PasiveUpgrade_TextBlock : text_block = text_block{DefaultTextColor:=UpgradeTextColor, DefaultShadowOffset:=ShadowOffset, DefaultShadowOpacity := DefaultTextShadowOpacity}
    P_Row4_PasiveUpgrade_TextBlock : text_block = text_block{DefaultTextColor:=UpgradeTextColor, DefaultShadowOffset:=ShadowOffset, DefaultShadowOpacity := DefaultTextShadowOpacity}

    
    P_Row1_PasiveUpgradeCost_TextBlock : text_block = text_block{DefaultTextColor:=AgedSapphiresTextColor, DefaultShadowOffset:=ShadowOffset, DefaultShadowOpacity := DefaultTextShadowOpacity}
    P_Row2_PasiveUpgradeCost_TextBlock : text_block = text_block{DefaultTextColor:=AgedSapphiresTextColor, DefaultShadowOffset:=ShadowOffset, DefaultShadowOpacity := DefaultTextShadowOpacity}
    P_Row3_PasiveUpgradeCost_TextBlock : text_block = text_block{DefaultTextColor:=AgedSapphiresTextColor, DefaultShadowOffset:=ShadowOffset, DefaultShadowOpacity := DefaultTextShadowOpacity}
    P_Row4_PasiveUpgradeCost_TextBlock : text_block = text_block{DefaultTextColor:=AgedSapphiresTextColor, DefaultShadowOffset:=ShadowOffset, DefaultShadowOpacity := DefaultTextShadowOpacity}

    R_Row1CapacityButton : button_quiet = button_quiet{}
    R_Row2CapacityButton : button_quiet = button_quiet{}
    R_Row3CapacityButton : button_quiet = button_quiet{}
    R_Row4CapacityButton : button_quiet = button_quiet{}

    P_Row1PasiveButton : button_quiet = button_quiet{}
    P_Row2PasiveButton : button_quiet = button_quiet{}
    P_Row3PasiveButton : button_quiet = button_quiet{}
    P_Row4PasiveButton : button_quiet = button_quiet{}

    var AllResourcersColor : []color = array{}

    var AllR_Rows_Capacity_TextBlock : []text_block = array{}
    var AllR_Rows_CapacityUpgrade_TextBlock : []text_block = array{}
    var AllR_Rows_CapacityUpgradeCost_TextBlock : []text_block = array{}

    var AllP_Rows_Pasive_TextBlock : []text_block = array{}
    var AllP_Rows_PasiveUpgrade_TextBlock : []text_block = array{}
    var AllP_Rows_PasiveUpgradeCost_TextBlock : []text_block = array{}

    CloseButton : button_loud = button_loud{}

    var MaybeCanvas : ?canvas = false
    var UIIsDisplayed : logic = false

    var TimeElapsed : int = 0
    var BuyModeMultiplier : int = 1

    InitUI() : void =
        SetArrays()
        InitSubscribers()

        Canvas := MakeCanvas()
        set MaybeCanvas = option{Canvas}

    SetArrays() : void =

        set AllResourcersColor = array{WaterTextColor, FlourTextColor, CheeseTextColor, PimentoTextColor}
        set AllR_Rows_Capacity_TextBlock = array{R_Row1_Capacity_TextBlock,R_Row2_Capacity_TextBlock,R_Row3_Capacity_TextBlock,R_Row4_Capacity_TextBlock}
        set AllR_Rows_CapacityUpgrade_TextBlock = array{R_Row1_CapacityUpgrade_TextBlock,R_Row2_CapacityUpgrade_TextBlock,R_Row3_CapacityUpgrade_TextBlock,R_Row4_CapacityUpgrade_TextBlock}
        set AllR_Rows_CapacityUpgradeCost_TextBlock = array{R_Row1_CapacityUpgradeCost_TextBlock,R_Row2_CapacityUpgradeCost_TextBlock,R_Row3_CapacityUpgradeCost_TextBlock,R_Row4_CapacityUpgradeCost_TextBlock}

        set AllP_Rows_Pasive_TextBlock = array{P_Row1_Pasive_TextBlock,P_Row2_Pasive_TextBlock,P_Row3_Pasive_TextBlock,P_Row4_Pasive_TextBlock}
        set AllP_Rows_PasiveUpgrade_TextBlock = array{P_Row1_PasiveUpgrade_TextBlock,P_Row2_PasiveUpgrade_TextBlock,P_Row3_PasiveUpgrade_TextBlock,P_Row4_PasiveUpgrade_TextBlock}
        set AllP_Rows_PasiveUpgradeCost_TextBlock = array{P_Row1_PasiveUpgradeCost_TextBlock,P_Row2_PasiveUpgradeCost_TextBlock,P_Row3_PasiveUpgradeCost_TextBlock,P_Row4_PasiveUpgradeCost_TextBlock}
    
    InitSubscribers() : void =

        R_Row1CapacityButton.OnClick().Subscribe(handler_upgrades{MaybeBakeryMachinesManager := option{BakeryMachinesManager}, Row:=0}.HandlerDepositUpgrade)  
        R_Row2CapacityButton.OnClick().Subscribe(handler_upgrades{MaybeBakeryMachinesManager := option{BakeryMachinesManager}, Row:=1}.HandlerDepositUpgrade)  
        R_Row3CapacityButton.OnClick().Subscribe(handler_upgrades{MaybeBakeryMachinesManager := option{BakeryMachinesManager}, Row:=2}.HandlerDepositUpgrade) 
        R_Row4CapacityButton.OnClick().Subscribe(handler_upgrades{MaybeBakeryMachinesManager := option{BakeryMachinesManager}, Row:=3}.HandlerDepositUpgrade) 

        P_Row1PasiveButton.OnClick().Subscribe(handler_upgrades{MaybeBakeryMachinesManager := option{BakeryMachinesManager}, Row:=0}.HandlerDepositPasiveUpgrade)
        P_Row2PasiveButton.OnClick().Subscribe(handler_upgrades{MaybeBakeryMachinesManager := option{BakeryMachinesManager}, Row:=1}.HandlerDepositPasiveUpgrade)  
        P_Row3PasiveButton.OnClick().Subscribe(handler_upgrades{MaybeBakeryMachinesManager := option{BakeryMachinesManager}, Row:=2}.HandlerDepositPasiveUpgrade)    
        P_Row4PasiveButton.OnClick().Subscribe(handler_upgrades{MaybeBakeryMachinesManager := option{BakeryMachinesManager}, Row:=3}.HandlerDepositPasiveUpgrade)  
        
        CloseButton.OnClick().Subscribe(Remove_Canvas) 

    SetAgent(InAgent : agent) : void =
        set MaybeAgent = option{InAgent}

    UpdateUI(PlayerData : player_data_table) : void =

        CloseButton.SetText(BasicMessage("CLOSE"))

        #Inventory resourcers
        TotalWaterAvailable := PlayerData.Inventory_Water + PlayerData.MainWaterDeposit_Value + PlayerData.WaterPreparedFor_B + PlayerData.WaterPreparedFor_CB + PlayerData.WaterPreparedFor_PCS
        TotalFlourAvailable := PlayerData.Inventory_Flour + PlayerData.MainFlourDeposit_Value + PlayerData.FlourPreparedFor_B + PlayerData.FlourPreparedFor_CB + PlayerData.FlourPreparedFor_PCS
        TotalCheeseAvailable := PlayerData.Inventory_Cheese + PlayerData.MainCheeseDeposit_Value + PlayerData.CheesePreparedFor_CB + PlayerData.CheesePreparedFor_PCS
        TotalPimentosAvailable := PlayerData.Inventory_Pimentos + PlayerData.MainPimentosDeposit_Value + PlayerData.PimentosPreparedFor_PCS

        TotalSapphires_TextBlock.SetText(BasicMessage(GetMagnitudeLetter(PlayerData.Inventory_Sapphires,"", "")))
        TotalAgedSapphires_TextBlock.SetText(BasicMessage(GetMagnitudeLetter(PlayerData.Inventory_AgedSapphires,"", "")))
        TotalWater_TextBlock.SetText(BasicMessage(GetMagnitudeLetter(TotalWaterAvailable,"", "")))
        TotalFlour_TextBlock.SetText(BasicMessage(GetMagnitudeLetter(TotalFlourAvailable,"", "")))
        TotalCheese_TextBlock.SetText(BasicMessage(GetMagnitudeLetter(TotalCheeseAvailable,"", "")))
        TotalPimentos_TextBlock.SetText(BasicMessage(GetMagnitudeLetter(TotalPimentosAvailable,"", "")))
        ArrayResourcers:=array{PlayerData.Inventory_Water,PlayerData.Inventory_Flour,PlayerData.Inventory_Cheese,PlayerData.Inventory_Pimentos}

        #AllUP_Rows_Level_TextBlock
        for(Resource:=0..4):
            if:      
                R_Row_Capacity_TextBlock:=AllR_Rows_Capacity_TextBlock[Resource]
                R_Row_CapacityUpgrade_TextBlock:=AllR_Rows_CapacityUpgrade_TextBlock[Resource]
                R_Row_CapacityUpgradeCost_TextBlock:=AllR_Rows_CapacityUpgradeCost_TextBlock[Resource]
                
                P_Row_Pasive_TextBlock := AllP_Rows_Pasive_TextBlock[Resource]
                P_Row_PasiveUpgrade_TextBlock := AllP_Rows_PasiveUpgrade_TextBlock[Resource]
                P_Row_PasiveUpgradeCost_TextBlock := AllP_Rows_PasiveUpgradeCost_TextBlock[Resource]

                Text_Color := AllResourcersColor[Resource]

                PlayerResourceInventory := ArrayResourcers[Resource]
                PlayerAgedSapphires := PlayerData.Inventory_AgedSapphires
            then:
                if:
                    ResLimitLV := PlayerData.AllLV_MainDeposits[Resource]
                    PasiveLV := PlayerData.AllLV_PasiveResourcers[Resource]
                then:
                    DupleResourceDepositLimit:=CalculeDepositUpgradeLimit(ResLimitLV, Resource)
                    DuplePasiveIncome:=CalculeDepositPasiveIncome(PasiveLV)

                    if(PasiveLV<MaxUpgradePasiveIncome):
                        P_Row_Pasive_TextBlock.SetText(BasicMessage(GetMagnitudeLetter(DuplePasiveIncome(0)*100.0,"","%")))
                        P_Row_PasiveUpgrade_TextBlock.SetText(BasicMessage(GetMagnitudeLetter(DuplePasiveIncome(2)*100.0,"+","%")))
                        P_Row_PasiveUpgrade_TextBlock.SetTextColor(UpgradeTextColor)
                        P_Row_PasiveUpgradeCost_TextBlock.SetText(BasicMessage(GetMagnitudeLetter(DuplePasiveIncome(1),"x ","")))
                        P_Row_PasiveUpgradeCost_TextBlock.SetTextColor(AgedSapphiresTextColor)

                        if(PlayerData.Inventory_AgedSapphires<DuplePasiveIncome(1)):
                            P_Row_PasiveUpgradeCost_TextBlock.SetTextColor(NotEnoughTextColor)
                        else:
                            P_Row_PasiveUpgradeCost_TextBlock.SetTextColor(AgedSapphiresTextColor)
                    else:
                        P_Row_Pasive_TextBlock.SetText(BasicMessage(GetMagnitudeLetter(DuplePasiveIncome(0)*100.0,"","%")))
                        P_Row_PasiveUpgrade_TextBlock.SetText(BasicMessage("MAXED"))
                        P_Row_PasiveUpgrade_TextBlock.SetTextColor(MaxedColor)
                        P_Row_PasiveUpgradeCost_TextBlock.SetText(BasicMessage("MAXED"))
                        P_Row_PasiveUpgradeCost_TextBlock.SetTextColor(MaxedColor)

                    R_Row_CapacityUpgradeCost_TextBlock.SetTextColor(Text_Color)
                    R_Row_Capacity_TextBlock.SetText(BasicMessage(GetOptimizedMagnitudeLetter(DupleResourceDepositLimit(0))))
                    R_Row_CapacityUpgradeCost_TextBlock.SetText(BasicMessage(GetMagnitudeLetter(DupleResourceDepositLimit(1),"x ",""))) 
                    R_Row_CapacityUpgrade_TextBlock.SetText(BasicMessage(GetMagnitudeLetter(DupleResourceDepositLimit(2),"+","")))

                        <#
                    if(PlayerResourceInventory<DupleResourceDepositLimit(1)):
                        R_Row_CapacityUpgradeCost_TextBlock.SetTextColor(NotEnoughTextColor)
                    else:
                        R_Row_CapacityUpgradeCost_TextBlock.SetTextColor(Text_Color)#>

                    TupleHasEnough:=BakeryMachinesManager.PlayersDataManager.HasEnoughResource(PlayerData,DupleResourceDepositLimit(1),Resource)
                    if(TupleHasEnough(0)<0):
                        R_Row_CapacityUpgradeCost_TextBlock.SetTextColor(NotEnoughTextColor)
                    else:
                        R_Row_CapacityUpgradeCost_TextBlock.SetTextColor(Text_Color)             
                    

    Remove_Canvas(Message : widget_message) : void =
        set UIIsDisplayed = false
        if:
            Agent:=MaybeAgent?
            Player:=player[Agent]
            PlayerUI:=GetPlayerUI[Player]
            Canvas := MaybeCanvas?
        then:
            PlayerUI.RemoveWidget(Canvas)
            RemoveDelay_OnRemove()

    Remove_CanvasWithoutMessage() : void =
        set UIIsDisplayed = false
        #Print("Trying to remove UI")
        if:
            Agent:=MaybeAgent?
            Player:=player[Agent]
            PlayerUI:=GetPlayerUI[Player]
            Canvas := MaybeCanvas?
        then:
            #Print("Removing UI")
            PlayerUI.RemoveWidget(Canvas)
            RemoveDelay_OnRemove()

    RemoveDelay_OnAdd() : void =
        TitleText.SetText(BasicMessage("Capacity deposits and pasive income"))
        #CurrentBuyOption_textblock.SetVisibility(widget_visibility.Hidden)

    RemoveDelay_OnRemove() : void =
        TitleText.SetText(BasicMessage("Capacity deposits and pasive income "))
        #CurrentBuyOption_textblock.SetVisibility(widget_visibility.Hidden)

            
    Display_Canvas() : void =
        set TimeElapsed = 0
        set UIIsDisplayed = true
        spawn: 
            StartTimeElapsed()
        if:
            Agent:=MaybeAgent?
            PlayerUI:=GetPlayerUI[player[Agent]]
            Canvas := MaybeCanvas?
        then:
            ##Print("Assigning created UI before")
            PlayerUI.AddWidget(Canvas,player_ui_slot{InputMode := ui_input_mode.All, ZOrder:=1}) 
            RemoveDelay_OnAdd() 

    StartTimeElapsed()<suspends> : void =
        loop:
            ##Print("IN THE LOOP TimeElapsed: {TimeElapsed}")
            if(TimeElapsed > MaxUIDisplayTime):
                ##Print("Closing the UI because 20 seconds has pased")
                Remove_CanvasWithoutMessage()
                break
            if(UIIsDisplayed = false):
                ##Print("Breaking Time Elapsed loop")
                break
            Sleep(1.0)
            set TimeElapsed = TimeElapsed + 1




    MakeCanvas(): canvas=
        Canvas := canvas:
            Slots:=array:
                #Background black image
                canvas_slot:
                    Anchors:=anchors{Minimum:=vector2{X:=0.5,Y:=0.5},Maximum:=vector2{X:=0.5,Y:=0.5}}
                    ZOrder:=0
                    Alignment:=vector2{X:=0.5,Y:=0.5}
                    Widget:=texture_block:
                        DefaultImage:=Images.BlackBackground
                        DefaultDesiredSize:=vector2{X:=BackgroundBlackUISizeX,Y:=BackgroundBlackUISizeY}
                canvas_slot: #Background image
                    Anchors:=anchors{Minimum:=vector2{X:=0.5,Y:=0.5},Maximum:=vector2{X:=0.5,Y:=0.5}}
                    ZOrder:=0
                    Offsets:=margin{Left:=0.0,Top:=0.0,Right:=BackgroundImageSizeX,Bottom:=BackgroundImageSizeY} #PositionX, PositionY, SizeX, SizeY
                    Alignment:=vector2{X:=0.5,Y:=0.5}
                    Widget:=texture_block:
                        DefaultImage:=Images.Img_MainBackground
                        DefaultDesiredSize:=vector2{X:=BackgroundImageSizeX,Y:=BackgroundImageSizeY}
                canvas_slot: #Main overlay
                    Anchors:=anchors{Minimum:=vector2{X:=0.5,Y:=0.5},Maximum:=vector2{X:=0.5,Y:=0.5}}
                    ZOrder:=0
                    #Offsets:=margin{Left:=1920.0,Top:=0.0,Right:=0.0,Bottom:=1080.0} #PositionX, PositionY, SizeX, SizeY
                    Alignment:=vector2{X:=0.5,Y:=0.5}
                    Widget:=overlay:
                        Slots:=array:
                            overlay_slot:
                                HorizontalAlignment:=horizontal_alignment.Center
                                VerticalAlignment:=vertical_alignment.Center
                                Widget:=stack_box:
                                    Orientation:=orientation.Vertical
                                        Slots:=array:
                                            stack_box_slot: #Resourcers top show 
                                                Padding:=margin{Left:=RowLeftPadding,Top:=RowTopPadding,Right:=RowRightPadding,Bottom:=RowBottomPadding}
                                                Widget:=stack_box:
                                                    Orientation:=orientation.Horizontal
                                                        Slots:=array:
                                                            stack_box_slot: #Empty texture
                                                                HorizontalAlignment:=horizontal_alignment.Fill
                                                                VerticalAlignment:=vertical_alignment.Fill
                                                                Padding:=margin{Left:=InterColumnLeftPadding,Top:=InterColumnTopPadding,Right:=InterColumnRightPadding,Bottom:=InterColumnBottomPadding}
                                                                Widget:=overlay:
                                                                    Slots:=array:
                                                                        overlay_slot:
                                                                            Widget:=texture_block{DefaultImage:=Images.EmptyTexture, DefaultDesiredSize:=vector2{X:=595.0,Y:=50.0}}
                                                            stack_box_slot: #Water
                                                                HorizontalAlignment:=horizontal_alignment.Fill
                                                                VerticalAlignment:=vertical_alignment.Fill
                                                                Padding:=margin{Left:=InterColumnLeftPadding,Top:=InterColumnTopPadding,Right:=InterColumnRightPadding,Bottom:=InterColumnBottomPadding}
                                                                Widget:=overlay:
                                                                    Slots:=array:
                                                                        overlay_slot:
                                                                            Widget:=TotalWater_BackgroundTexture
                                                                        overlay_slot: 
                                                                            HorizontalAlignment:=horizontal_alignment.Left
                                                                            VerticalAlignment:=vertical_alignment.Center 
                                                                            Padding:=margin{Left:=TextIcon_LeftMargin}
                                                                            Widget:=TotalWater_TextBlock 
                                                                        overlay_slot: 
                                                                            HorizontalAlignment:=horizontal_alignment.Left
                                                                            VerticalAlignment:=vertical_alignment.Center 
                                                                            Padding:=margin{Left:=Icon_LeftMargin}
                                                                            Widget:=texture_block{DefaultDesiredSize:=vector2{X:=IconSmallSizeX,Y:=IconSmallSizeY}, DefaultImage:=Images.Water_Icon}
                                                            stack_box_slot: #Flour
                                                                HorizontalAlignment:=horizontal_alignment.Fill
                                                                VerticalAlignment:=vertical_alignment.Fill
                                                                Padding:=margin{Left:=InterColumnLeftPadding,Top:=InterColumnTopPadding,Right:=InterColumnRightPadding,Bottom:=InterColumnBottomPadding}
                                                                Widget:=overlay:
                                                                    Slots:=array:
                                                                        overlay_slot:
                                                                            Widget:=TotalFlour_BackgroundTexture
                                                                        overlay_slot: 
                                                                            HorizontalAlignment:=horizontal_alignment.Left
                                                                            VerticalAlignment:=vertical_alignment.Center 
                                                                            Padding:=margin{Left:=TextIcon_LeftMargin}
                                                                            Widget:=TotalFlour_TextBlock 
                                                                        overlay_slot: 
                                                                            HorizontalAlignment:=horizontal_alignment.Left
                                                                            VerticalAlignment:=vertical_alignment.Center 
                                                                            Padding:=margin{Left:=Icon_LeftMargin}
                                                                            Widget:=texture_block{DefaultDesiredSize:=vector2{X:=IconSmallSizeX,Y:=IconSmallSizeY}, DefaultImage:=Images.Flour_Icon}
                                                            stack_box_slot: #Cheese
                                                                HorizontalAlignment:=horizontal_alignment.Fill
                                                                VerticalAlignment:=vertical_alignment.Fill
                                                                Padding:=margin{Left:=InterColumnLeftPadding,Top:=InterColumnTopPadding,Right:=InterColumnRightPadding,Bottom:=InterColumnBottomPadding}
                                                                Widget:=overlay:
                                                                    Slots:=array:
                                                                        overlay_slot:
                                                                            Widget:=TotalCheese_BackgroundTexture
                                                                        overlay_slot: 
                                                                            HorizontalAlignment:=horizontal_alignment.Left
                                                                            VerticalAlignment:=vertical_alignment.Center 
                                                                            Padding:=margin{Left:=TextIcon_LeftMargin}
                                                                            Widget:=TotalCheese_TextBlock 
                                                                        overlay_slot: 
                                                                            HorizontalAlignment:=horizontal_alignment.Left
                                                                            VerticalAlignment:=vertical_alignment.Center 
                                                                            Padding:=margin{Left:=Icon_LeftMargin}
                                                                            Widget:=texture_block{DefaultDesiredSize:=vector2{X:=IconSmallSizeX,Y:=IconSmallSizeY}, DefaultImage:=Images.Cheese_Icon}
                                                            stack_box_slot: #Pimientos
                                                                HorizontalAlignment:=horizontal_alignment.Fill
                                                                VerticalAlignment:=vertical_alignment.Fill
                                                                Padding:=margin{Left:=InterColumnLeftPadding,Top:=InterColumnTopPadding,Right:=InterColumnRightPadding,Bottom:=InterColumnBottomPadding}
                                                                Widget:=overlay:
                                                                    Slots:=array:
                                                                        overlay_slot:
                                                                            Widget:=TotalPimentos_BackgroundTexture
                                                                        overlay_slot: 
                                                                            HorizontalAlignment:=horizontal_alignment.Left
                                                                            VerticalAlignment:=vertical_alignment.Center 
                                                                            Padding:=margin{Left:=TextIcon_LeftMargin}
                                                                            Widget:=TotalPimentos_TextBlock 
                                                                        overlay_slot: 
                                                                            HorizontalAlignment:=horizontal_alignment.Left
                                                                            VerticalAlignment:=vertical_alignment.Center 
                                                                            Padding:=margin{Left:=Icon_LeftMargin}
                                                                            Widget:=texture_block{DefaultDesiredSize:=vector2{X:=IconSmallSizeX,Y:=IconSmallSizeY}, DefaultImage:=Images.Pimento_Icon}
                                                            stack_box_slot: #AgedSapphires
                                                                HorizontalAlignment:=horizontal_alignment.Fill
                                                                VerticalAlignment:=vertical_alignment.Fill
                                                                Padding:=margin{Left:=InterColumnLeftPadding,Top:=InterColumnTopPadding,Right:=InterColumnRightPadding,Bottom:=InterColumnBottomPadding}
                                                                Widget:=overlay:
                                                                    Slots:=array:
                                                                        overlay_slot:
                                                                            Widget:=TotalAgedSapphires_BackgroundTexture
                                                                        overlay_slot: 
                                                                            HorizontalAlignment:=horizontal_alignment.Left
                                                                            VerticalAlignment:=vertical_alignment.Center 
                                                                            Padding:=margin{Left:=TextIcon_LeftMargin}
                                                                            Widget:=TotalAgedSapphires_TextBlock 
                                                                        overlay_slot: 
                                                                            HorizontalAlignment:=horizontal_alignment.Left
                                                                            VerticalAlignment:=vertical_alignment.Center 
                                                                            Padding:=margin{Left:=Icon_LeftMargin}
                                                                            Widget:=texture_block{DefaultDesiredSize:=vector2{X:=IconSmallSizeX,Y:=IconSmallSizeY}, DefaultImage:=Images.AgedSaphire_Icon}
                                            stack_box_slot: #Complete list
                                                Padding:=margin{Left:=RowLeftPadding,Top:=RowTopPadding,Right:=RowRightPadding,Bottom:=RowBottomPadding}
                                                Widget:=stack_box:
                                                    Orientation:=orientation.Vertical
                                                        Slots:=array:
                                                            stack_box_slot: #Title
                                                                HorizontalAlignment:=horizontal_alignment.Fill
                                                                VerticalAlignment:=vertical_alignment.Fill
                                                                Padding:=margin{Left:=RowLeftPadding,Top:=RowTopPadding,Right:=RowRightPadding,Bottom:=RowBottomPadding}
                                                                Widget:=overlay:
                                                                    Slots:=array:
                                                                        overlay_slot: 
                                                                            HorizontalAlignment:=horizontal_alignment.Fill
                                                                            VerticalAlignment:=vertical_alignment.Bottom
                                                                            Widget:=texture_block{DefaultDesiredSize:=vector2{X:=HorizontalLineSizeX,Y:=HorizontalLineSizeY},DefaultImage:=Images.Img_HorizontalLine}
                                                                        overlay_slot: 
                                                                            HorizontalAlignment:=horizontal_alignment.Center
                                                                            VerticalAlignment:=vertical_alignment.Center 
                                                                            Widget:=TitleText
                                                                        overlay_slot: 
                                                                            HorizontalAlignment:=horizontal_alignment.Center
                                                                            VerticalAlignment:=vertical_alignment.Center 
                                                                            Widget:=texture_block{DefaultDesiredSize:=vector2{X:=0.0,Y:=RColumn1SizeY},DefaultImage:=Images.EmptyTexture}
                                                            stack_box_slot: #Title items
                                                                HorizontalAlignment:=horizontal_alignment.Fill
                                                                VerticalAlignment:=vertical_alignment.Fill
                                                                Padding:=margin{Left:=RowLeftPadding,Top:=RowTopPadding,Right:=RowRightPadding,Bottom:=RowBottomPadding}
                                                                Widget:=overlay:
                                                                    Slots:=array:
                                                                        overlay_slot: 
                                                                            HorizontalAlignment:=horizontal_alignment.Fill
                                                                            Widget:=texture_block{DefaultDesiredSize:=vector2{X:=0.0,Y:=LineSizeY}, DefaultImage:=Images.Img_LineBackground}
                                                                        overlay_slot: 
                                                                            HorizontalAlignment:=horizontal_alignment.Center
                                                                            VerticalAlignment:=vertical_alignment.Center 
                                                                            Widget:=stack_box:
                                                                                Orientation:=orientation.Horizontal
                                                                                    Slots:=array:
                                                                                        stack_box_slot: #Column2
                                                                                            Padding:=margin{Left:=InterColumnLeftPadding,Top:=InterColumnTopPadding,Right:=InterColumnRightPadding,Bottom:=InterColumnBottomPadding}
                                                                                            Widget:=overlay:
                                                                                                Slots:=array:
                                                                                                    overlay_slot: 
                                                                                                        Widget:=texture_block{DefaultDesiredSize:=vector2{X:=RColumn2SizeX,Y:=RColumn2SizeY}, DefaultImage:=Images.EmptyTexture}
                                                                                                    overlay_slot: 
                                                                                                        Widget:=text_block{DefaultText:=BasicMessage("Resource"), DefaultTextColor:=White, DefaultShadowOffset:=ShadowOffset, DefaultShadowOpacity := DefaultTextShadowOpacity}
                                                                                        stack_box_slot: #Column3
                                                                                            Padding:=margin{Left:=InterColumnLeftPadding,Top:=InterColumnTopPadding,Right:=InterColumnRightPadding,Bottom:=InterColumnBottomPadding}
                                                                                            Widget:=overlay:
                                                                                                Slots:=array:
                                                                                                    overlay_slot: 
                                                                                                        Widget:=texture_block{DefaultDesiredSize:=vector2{X:=RColumn3SizeX,Y:=RColumn3SizeY}, DefaultImage:=Images.EmptyTexture}
                                                                                                    overlay_slot: 
                                                                                                        Widget:=text_block{DefaultText:=BasicMessage("Capacity"), DefaultTextColor:=White, DefaultShadowOffset:=ShadowOffset, DefaultShadowOpacity := DefaultTextShadowOpacity}
                                                                                        stack_box_slot: #Column4
                                                                                            Padding:=margin{Left:=InterColumnLeftPadding,Top:=InterColumnTopPadding,Right:=InterColumnRightPadding,Bottom:=InterColumnBottomPadding}
                                                                                            Widget:=overlay:
                                                                                                Slots:=array:
                                                                                                    overlay_slot: 
                                                                                                        Widget:=texture_block{DefaultDesiredSize:=vector2{X:=RColumn4SizeX,Y:=RColumn4SizeY}, DefaultImage:=Images.EmptyTexture}
                                                                                                    overlay_slot: 
                                                                                                        Widget:=text_block{DefaultText:=BasicMessage("Upgrade"), DefaultTextColor:=White, DefaultShadowOffset:=ShadowOffset, DefaultShadowOpacity := DefaultTextShadowOpacity}
                                                                                        stack_box_slot: #Column5
                                                                                            Padding:=margin{Left:=InterColumnLeftPadding,Top:=InterColumnTopPadding,Right:=InterColumnRightPadding,Bottom:=InterColumnBottomPadding}
                                                                                            Widget:=overlay:
                                                                                                Slots:=array:
                                                                                                    overlay_slot: 
                                                                                                        Widget:=texture_block{DefaultDesiredSize:=vector2{X:=RColumn5SizeX,Y:=RColumn5SizeY}, DefaultImage:=Images.EmptyTexture}
                                                                                                    overlay_slot: 
                                                                                                        Widget:=text_block{DefaultText:=BasicMessage("Upgrade cost"), DefaultTextColor:=White, DefaultShadowOffset:=ShadowOffset, DefaultShadowOpacity := DefaultTextShadowOpacity}
                                                                                        stack_box_slot: #Column6
                                                                                            Padding:=margin{Left:=InterColumnLeftPadding,Top:=InterColumnTopPadding,Right:=InterColumnRightPadding,Bottom:=InterColumnBottomPadding}
                                                                                            Widget:=overlay:
                                                                                                Slots:=array:
                                                                                                    overlay_slot: 
                                                                                                        Widget:=texture_block{DefaultDesiredSize:=vector2{X:=RColumn6SizeX,Y:=RColumn6SizeY}, DefaultImage:=Images.EmptyTexture}
                                                                                                    overlay_slot: 
                                                                                                        Widget:=text_block{DefaultText:=BasicMessage("*Pasive income"), DefaultTextColor:=White, DefaultShadowOffset:=ShadowOffset, DefaultShadowOpacity := DefaultTextShadowOpacity}
                                                                                        stack_box_slot: #Column7
                                                                                            Padding:=margin{Left:=InterColumnLeftPadding,Top:=InterColumnTopPadding,Right:=InterColumnRightPadding,Bottom:=InterColumnBottomPadding}
                                                                                            Widget:=overlay:
                                                                                                Slots:=array:
                                                                                                    overlay_slot: 
                                                                                                        Widget:=texture_block{DefaultDesiredSize:=vector2{X:=RColumn7SizeX,Y:=RColumn7SizeY}, DefaultImage:=Images.EmptyTexture}
                                                                                                    overlay_slot: 
                                                                                                        Widget:=text_block{DefaultText:=BasicMessage("Upgrade"), DefaultTextColor:=White, DefaultShadowOffset:=ShadowOffset, DefaultShadowOpacity := DefaultTextShadowOpacity}
                                                                                        stack_box_slot: #Column8
                                                                                            Padding:=margin{Left:=InterColumnLeftPadding,Top:=InterColumnTopPadding,Right:=InterColumnRightPadding,Bottom:=InterColumnBottomPadding}
                                                                                            Widget:=overlay:
                                                                                                Slots:=array:
                                                                                                    overlay_slot: 
                                                                                                        Widget:=texture_block{DefaultDesiredSize:=vector2{X:=RColumn8SizeX,Y:=RColumn8SizeY}, DefaultImage:=Images.EmptyTexture}
                                                                                                    overlay_slot: 
                                                                                                        Widget:=text_block{DefaultText:=BasicMessage("Upgrade cost"), DefaultTextColor:=White, DefaultShadowOffset:=ShadowOffset, DefaultShadowOpacity := DefaultTextShadowOpacity}

                                                            stack_box_slot: #Water
                                                                HorizontalAlignment:=horizontal_alignment.Fill
                                                                VerticalAlignment:=vertical_alignment.Fill
                                                                Padding:=margin{Left:=RowLeftPadding,Top:=RowTopPadding,Right:=RowRightPadding,Bottom:=RowBottomPadding}
                                                                Widget:=overlay:
                                                                    Slots:=array:
                                                                        overlay_slot: 
                                                                            HorizontalAlignment:=horizontal_alignment.Fill
                                                                            Widget:=texture_block{DefaultDesiredSize:=vector2{X:=0.0,Y:=LineSizeY}, DefaultImage:=Images.Img_LineBackground}
                                                                        overlay_slot: 
                                                                            HorizontalAlignment:=horizontal_alignment.Center
                                                                            VerticalAlignment:=vertical_alignment.Center 
                                                                            Widget:=stack_box:
                                                                                Orientation:=orientation.Horizontal
                                                                                    Slots:=array:                     
                                                                                        stack_box_slot: #Column2
                                                                                            Padding:=margin{Left:=InterColumnLeftPadding,Top:=InterColumnTopPadding,Right:=InterColumnRightPadding,Bottom:=InterColumnBottomPadding}
                                                                                            Widget:=overlay:
                                                                                                Slots:=array:
                                                                                                    overlay_slot: 
                                                                                                        Widget:=texture_block{DefaultDesiredSize:=vector2{X:=RColumn2SizeX,Y:=RColumn2SizeY}, DefaultImage:=Images.EmptyTexture}
                                                                                                    overlay_slot: 
                                                                                                        Widget:=texture_block{DefaultDesiredSize:=vector2{X:=IconSizeX,Y:=IconSizeY}, DefaultImage:=Images.Water_Icon}
                                                                                        stack_box_slot: #Column3
                                                                                            Padding:=margin{Left:=InterColumnLeftPadding,Top:=InterColumnTopPadding,Right:=InterColumnRightPadding,Bottom:=InterColumnBottomPadding}
                                                                                            Widget:=overlay:
                                                                                                Slots:=array:
                                                                                                    overlay_slot: 
                                                                                                        Widget:=texture_block{DefaultDesiredSize:=vector2{X:=RColumn3SizeX,Y:=RColumn3SizeY}, DefaultImage:=Images.EmptyTexture}
                                                                                                    overlay_slot: 
                                                                                                        Widget:=R_Row1_Capacity_TextBlock
                                                                                        stack_box_slot: #Column4
                                                                                            Padding:=margin{Left:=InterColumnLeftPadding,Top:=InterColumnTopPadding,Right:=InterColumnRightPadding,Bottom:=InterColumnBottomPadding}
                                                                                            Widget:=overlay:
                                                                                                Slots:=array:
                                                                                                    overlay_slot: 
                                                                                                        Widget:=texture_block{DefaultDesiredSize:=vector2{X:=RColumn4SizeX,Y:=RColumn4SizeY}, DefaultImage:=Images.EmptyTexture}
                                                                                                    overlay_slot: 
                                                                                                        Widget:=R_Row1_CapacityUpgrade_TextBlock
                                                                                        stack_box_slot: #Column5
                                                                                            Padding:=margin{Left:=InterColumnLeftPadding,Top:=InterColumnTopPadding,Right:=InterColumnRightPadding,Bottom:=InterColumnBottomPadding}
                                                                                            Widget:=overlay:
                                                                                                Slots:=array:
                                                                                                    overlay_slot: 
                                                                                                        Widget:=texture_block{DefaultDesiredSize:=vector2{X:=RColumn5SizeX,Y:=RColumn5SizeY}, DefaultImage:=Images.EmptyTexture} 
                                                                                                    overlay_slot: 
                                                                                                        HorizontalAlignment:=horizontal_alignment.Left
                                                                                                        VerticalAlignment:=vertical_alignment.Center 
                                                                                                        Padding:=margin{Left:=TextIcon_LeftMargin}
                                                                                                        Widget:=R_Row1_CapacityUpgradeCost_TextBlock               
                                                                                                    overlay_slot: 
                                                                                                        HorizontalAlignment:=horizontal_alignment.Left
                                                                                                        VerticalAlignment:=vertical_alignment.Center 
                                                                                                        Padding:=margin{Left:=Icon_LeftMargin}
                                                                                                        Widget:=texture_block{DefaultDesiredSize:=vector2{X:=IconSmallSizeX,Y:=IconSmallSizeY}, DefaultImage:=Images.Water_Icon}
                                                                                                    overlay_slot:
                                                                                                        Padding:=margin{Left:=InterButtonLeftPadding,Top:=InterButtonTopPadding,Right:=InterButtonRightPadding,Bottom:=InterButtonBottomPadding}
                                                                                                        HorizontalAlignment:=horizontal_alignment.Fill
                                                                                                        VerticalAlignment:=vertical_alignment.Fill 
                                                                                                        Widget:=R_Row1CapacityButton
                                                                                        stack_box_slot: #Column6
                                                                                            Padding:=margin{Left:=InterColumnLeftPadding,Top:=InterColumnTopPadding,Right:=InterColumnRightPadding,Bottom:=InterColumnBottomPadding}
                                                                                            Widget:=overlay:
                                                                                                Slots:=array:
                                                                                                    overlay_slot: 
                                                                                                        Widget:=texture_block{DefaultDesiredSize:=vector2{X:=RColumn6SizeX,Y:=RColumn6SizeY}, DefaultImage:=Images.EmptyTexture}
                                                                                                    overlay_slot: 
                                                                                                        Widget:=P_Row1_Pasive_TextBlock
                                                                                        stack_box_slot: #Column7
                                                                                            Padding:=margin{Left:=InterColumnLeftPadding,Top:=InterColumnTopPadding,Right:=InterColumnRightPadding,Bottom:=InterColumnBottomPadding}
                                                                                            Widget:=overlay:
                                                                                                Slots:=array:
                                                                                                    overlay_slot: 
                                                                                                        Widget:=texture_block{DefaultDesiredSize:=vector2{X:=RColumn7SizeX,Y:=RColumn7SizeY}, DefaultImage:=Images.EmptyTexture}
                                                                                                    overlay_slot: 
                                                                                                        Widget:=P_Row1_PasiveUpgrade_TextBlock
                                                                                        stack_box_slot: #Column8
                                                                                            Padding:=margin{Left:=InterColumnLeftPadding,Top:=InterColumnTopPadding,Right:=InterColumnRightPadding,Bottom:=InterColumnBottomPadding}
                                                                                            Widget:=overlay:
                                                                                                Slots:=array:
                                                                                                    overlay_slot: 
                                                                                                        Widget:=texture_block{DefaultDesiredSize:=vector2{X:=RColumn8SizeX,Y:=RColumn8SizeY}, DefaultImage:=Images.EmptyTexture}
                                                                                                    overlay_slot: 
                                                                                                        HorizontalAlignment:=horizontal_alignment.Left
                                                                                                        VerticalAlignment:=vertical_alignment.Center 
                                                                                                        Padding:=margin{Left:=TextIcon_LeftMargin}
                                                                                                        Widget:=P_Row1_PasiveUpgradeCost_TextBlock               
                                                                                                    overlay_slot: 
                                                                                                        HorizontalAlignment:=horizontal_alignment.Left
                                                                                                        VerticalAlignment:=vertical_alignment.Center 
                                                                                                        Padding:=margin{Left:=Icon_LeftMargin}
                                                                                                        Widget:=texture_block{DefaultDesiredSize:=vector2{X:=IconSmallSizeX,Y:=IconSmallSizeY}, DefaultImage:=Images.AgedSaphire_Icon}
                                                                                                    overlay_slot:
                                                                                                        Padding:=margin{Left:=InterButtonLeftPadding,Top:=InterButtonTopPadding,Right:=InterButtonRightPadding,Bottom:=InterButtonBottomPadding}
                                                                                                        HorizontalAlignment:=horizontal_alignment.Fill
                                                                                                        VerticalAlignment:=vertical_alignment.Fill  
                                                                                                        Widget:=P_Row1PasiveButton
                                                            stack_box_slot: #Flour
                                                                HorizontalAlignment:=horizontal_alignment.Fill
                                                                VerticalAlignment:=vertical_alignment.Fill
                                                                Padding:=margin{Left:=RowLeftPadding,Top:=RowTopPadding,Right:=RowRightPadding,Bottom:=RowBottomPadding}
                                                                Widget:=overlay:
                                                                    Slots:=array:
                                                                        overlay_slot: 
                                                                            HorizontalAlignment:=horizontal_alignment.Fill
                                                                            Widget:=texture_block{DefaultDesiredSize:=vector2{X:=0.0,Y:=LineSizeY}, DefaultImage:=Images.Img_LineBackground}
                                                                        overlay_slot: 
                                                                            HorizontalAlignment:=horizontal_alignment.Center
                                                                            VerticalAlignment:=vertical_alignment.Center 
                                                                            Widget:=stack_box:
                                                                                Orientation:=orientation.Horizontal
                                                                                    Slots:=array:
                                                                                        stack_box_slot: #Column2
                                                                                            Padding:=margin{Left:=InterColumnLeftPadding,Top:=InterColumnTopPadding,Right:=InterColumnRightPadding,Bottom:=InterColumnBottomPadding}
                                                                                            Widget:=overlay:
                                                                                                Slots:=array:
                                                                                                    overlay_slot: 
                                                                                                        Widget:=texture_block{DefaultDesiredSize:=vector2{X:=RColumn2SizeX,Y:=RColumn2SizeY}, DefaultImage:=Images.EmptyTexture}
                                                                                                    overlay_slot: 
                                                                                                        Widget:=texture_block{DefaultDesiredSize:=vector2{X:=IconSizeX,Y:=IconSizeY}, DefaultImage:=Images.Flour_Icon}
                                                                                        stack_box_slot: #Column3
                                                                                            Padding:=margin{Left:=InterColumnLeftPadding,Top:=InterColumnTopPadding,Right:=InterColumnRightPadding,Bottom:=InterColumnBottomPadding}
                                                                                            Widget:=overlay:
                                                                                                Slots:=array:
                                                                                                    overlay_slot: 
                                                                                                        Widget:=texture_block{DefaultDesiredSize:=vector2{X:=RColumn3SizeX,Y:=RColumn3SizeY}, DefaultImage:=Images.EmptyTexture}
                                                                                                    overlay_slot: 
                                                                                                        Widget:=R_Row2_Capacity_TextBlock
                                                                                        stack_box_slot: #Column4
                                                                                            Padding:=margin{Left:=InterColumnLeftPadding,Top:=InterColumnTopPadding,Right:=InterColumnRightPadding,Bottom:=InterColumnBottomPadding}
                                                                                            Widget:=overlay:
                                                                                                Slots:=array:
                                                                                                    overlay_slot: 
                                                                                                        Widget:=texture_block{DefaultDesiredSize:=vector2{X:=RColumn4SizeX,Y:=RColumn4SizeY}, DefaultImage:=Images.EmptyTexture}
                                                                                                    overlay_slot: 
                                                                                                        Widget:=R_Row2_CapacityUpgrade_TextBlock
                                                                                        stack_box_slot: #Column5
                                                                                            Padding:=margin{Left:=InterColumnLeftPadding,Top:=InterColumnTopPadding,Right:=InterColumnRightPadding,Bottom:=InterColumnBottomPadding}
                                                                                            Widget:=overlay:
                                                                                                Slots:=array:
                                                                                                    overlay_slot: 
                                                                                                        Widget:=texture_block{DefaultDesiredSize:=vector2{X:=RColumn5SizeX,Y:=RColumn5SizeY}, DefaultImage:=Images.EmptyTexture} 
                                                                                                    overlay_slot: 
                                                                                                        HorizontalAlignment:=horizontal_alignment.Left
                                                                                                        VerticalAlignment:=vertical_alignment.Center 
                                                                                                        Padding:=margin{Left:=TextIcon_LeftMargin}
                                                                                                        Widget:=R_Row2_CapacityUpgradeCost_TextBlock               
                                                                                                    overlay_slot: 
                                                                                                        HorizontalAlignment:=horizontal_alignment.Left
                                                                                                        VerticalAlignment:=vertical_alignment.Center 
                                                                                                        Padding:=margin{Left:=Icon_LeftMargin}
                                                                                                        Widget:=texture_block{DefaultDesiredSize:=vector2{X:=IconSmallSizeX,Y:=IconSmallSizeY}, DefaultImage:=Images.Flour_Icon}
                                                                                                    overlay_slot:
                                                                                                        Padding:=margin{Left:=InterButtonLeftPadding,Top:=InterButtonTopPadding,Right:=InterButtonRightPadding,Bottom:=InterButtonBottomPadding}
                                                                                                        HorizontalAlignment:=horizontal_alignment.Fill
                                                                                                        VerticalAlignment:=vertical_alignment.Fill 
                                                                                                        Widget:=R_Row2CapacityButton
                                                                                        stack_box_slot: #Column6
                                                                                            Padding:=margin{Left:=InterColumnLeftPadding,Top:=InterColumnTopPadding,Right:=InterColumnRightPadding,Bottom:=InterColumnBottomPadding}
                                                                                            Widget:=overlay:
                                                                                                Slots:=array:
                                                                                                    overlay_slot: 
                                                                                                        Widget:=texture_block{DefaultDesiredSize:=vector2{X:=RColumn6SizeX,Y:=RColumn6SizeY}, DefaultImage:=Images.EmptyTexture}
                                                                                                    overlay_slot: 
                                                                                                        Widget:=P_Row2_Pasive_TextBlock
                                                                                        stack_box_slot: #Column7
                                                                                            Padding:=margin{Left:=InterColumnLeftPadding,Top:=InterColumnTopPadding,Right:=InterColumnRightPadding,Bottom:=InterColumnBottomPadding}
                                                                                            Widget:=overlay:
                                                                                                Slots:=array:
                                                                                                    overlay_slot: 
                                                                                                        Widget:=texture_block{DefaultDesiredSize:=vector2{X:=RColumn7SizeX,Y:=RColumn7SizeY}, DefaultImage:=Images.EmptyTexture}
                                                                                                    overlay_slot: 
                                                                                                        Widget:=P_Row2_PasiveUpgrade_TextBlock
                                                                                        stack_box_slot: #Column8
                                                                                            Padding:=margin{Left:=InterColumnLeftPadding,Top:=InterColumnTopPadding,Right:=InterColumnRightPadding,Bottom:=InterColumnBottomPadding}
                                                                                            Widget:=overlay:
                                                                                                Slots:=array:
                                                                                                    overlay_slot: 
                                                                                                        Widget:=texture_block{DefaultDesiredSize:=vector2{X:=RColumn8SizeX,Y:=RColumn8SizeY}, DefaultImage:=Images.EmptyTexture}
                                                                                                    overlay_slot: 
                                                                                                        HorizontalAlignment:=horizontal_alignment.Left
                                                                                                        VerticalAlignment:=vertical_alignment.Center 
                                                                                                        Padding:=margin{Left:=TextIcon_LeftMargin}
                                                                                                        Widget:=P_Row2_PasiveUpgradeCost_TextBlock               
                                                                                                    overlay_slot: 
                                                                                                        HorizontalAlignment:=horizontal_alignment.Left
                                                                                                        VerticalAlignment:=vertical_alignment.Center 
                                                                                                        Padding:=margin{Left:=Icon_LeftMargin}
                                                                                                        Widget:=texture_block{DefaultDesiredSize:=vector2{X:=IconSmallSizeX,Y:=IconSmallSizeY}, DefaultImage:=Images.AgedSaphire_Icon}
                                                                                                    overlay_slot:
                                                                                                        Padding:=margin{Left:=InterButtonLeftPadding,Top:=InterButtonTopPadding,Right:=InterButtonRightPadding,Bottom:=InterButtonBottomPadding}
                                                                                                        HorizontalAlignment:=horizontal_alignment.Fill
                                                                                                        VerticalAlignment:=vertical_alignment.Fill  
                                                                                                        Widget:=P_Row2PasiveButton
                                                            stack_box_slot: #Cheese
                                                                HorizontalAlignment:=horizontal_alignment.Fill
                                                                VerticalAlignment:=vertical_alignment.Fill
                                                                Padding:=margin{Left:=RowLeftPadding,Top:=RowTopPadding,Right:=RowRightPadding,Bottom:=RowBottomPadding}
                                                                Widget:=overlay:
                                                                    Slots:=array:
                                                                        overlay_slot: 
                                                                            HorizontalAlignment:=horizontal_alignment.Fill
                                                                            Widget:=texture_block{DefaultDesiredSize:=vector2{X:=0.0,Y:=LineSizeY}, DefaultImage:=Images.Img_LineBackground}
                                                                        overlay_slot: 
                                                                            HorizontalAlignment:=horizontal_alignment.Center
                                                                            VerticalAlignment:=vertical_alignment.Center 
                                                                            Widget:=stack_box:
                                                                                Orientation:=orientation.Horizontal
                                                                                    Slots:=array:          
                                                                                        stack_box_slot: #Column2
                                                                                            Padding:=margin{Left:=InterColumnLeftPadding,Top:=InterColumnTopPadding,Right:=InterColumnRightPadding,Bottom:=InterColumnBottomPadding}
                                                                                            Widget:=overlay:
                                                                                                Slots:=array:
                                                                                                    overlay_slot: 
                                                                                                        Widget:=texture_block{DefaultDesiredSize:=vector2{X:=RColumn2SizeX,Y:=RColumn2SizeY}, DefaultImage:=Images.EmptyTexture}
                                                                                                    overlay_slot: 
                                                                                                        Widget:=texture_block{DefaultDesiredSize:=vector2{X:=IconSizeX,Y:=IconSizeY}, DefaultImage:=Images.Cheese_Icon}
                                                                                        stack_box_slot: #Column3
                                                                                            Padding:=margin{Left:=InterColumnLeftPadding,Top:=InterColumnTopPadding,Right:=InterColumnRightPadding,Bottom:=InterColumnBottomPadding}
                                                                                            Widget:=overlay:
                                                                                                Slots:=array:
                                                                                                    overlay_slot: 
                                                                                                        Widget:=texture_block{DefaultDesiredSize:=vector2{X:=RColumn3SizeX,Y:=RColumn3SizeY}, DefaultImage:=Images.EmptyTexture}
                                                                                                    overlay_slot: 
                                                                                                        Widget:=R_Row3_Capacity_TextBlock
                                                                                        stack_box_slot: #Column4
                                                                                            Padding:=margin{Left:=InterColumnLeftPadding,Top:=InterColumnTopPadding,Right:=InterColumnRightPadding,Bottom:=InterColumnBottomPadding}
                                                                                            Widget:=overlay:
                                                                                                Slots:=array:
                                                                                                    overlay_slot: 
                                                                                                        Widget:=texture_block{DefaultDesiredSize:=vector2{X:=RColumn4SizeX,Y:=RColumn4SizeY}, DefaultImage:=Images.EmptyTexture}
                                                                                                    overlay_slot: 
                                                                                                        Widget:=R_Row3_CapacityUpgrade_TextBlock
                                                                                        stack_box_slot: #Column5
                                                                                            Padding:=margin{Left:=InterColumnLeftPadding,Top:=InterColumnTopPadding,Right:=InterColumnRightPadding,Bottom:=InterColumnBottomPadding}
                                                                                            Widget:=overlay:
                                                                                                Slots:=array:
                                                                                                    overlay_slot: 
                                                                                                        Widget:=texture_block{DefaultDesiredSize:=vector2{X:=RColumn5SizeX,Y:=RColumn5SizeY}, DefaultImage:=Images.EmptyTexture} 
                                                                                                    overlay_slot: 
                                                                                                        HorizontalAlignment:=horizontal_alignment.Left
                                                                                                        VerticalAlignment:=vertical_alignment.Center 
                                                                                                        Padding:=margin{Left:=TextIcon_LeftMargin}
                                                                                                        Widget:=R_Row3_CapacityUpgradeCost_TextBlock               
                                                                                                    overlay_slot: 
                                                                                                        HorizontalAlignment:=horizontal_alignment.Left
                                                                                                        VerticalAlignment:=vertical_alignment.Center 
                                                                                                        Padding:=margin{Left:=Icon_LeftMargin}
                                                                                                        Widget:=texture_block{DefaultDesiredSize:=vector2{X:=IconSmallSizeX,Y:=IconSmallSizeY}, DefaultImage:=Images.Cheese_Icon}
                                                                                                    overlay_slot:
                                                                                                        Padding:=margin{Left:=InterButtonLeftPadding,Top:=InterButtonTopPadding,Right:=InterButtonRightPadding,Bottom:=InterButtonBottomPadding}
                                                                                                        HorizontalAlignment:=horizontal_alignment.Fill
                                                                                                        VerticalAlignment:=vertical_alignment.Fill 
                                                                                                        Widget:=R_Row3CapacityButton
                                                                                        stack_box_slot: #Column6
                                                                                            Padding:=margin{Left:=InterColumnLeftPadding,Top:=InterColumnTopPadding,Right:=InterColumnRightPadding,Bottom:=InterColumnBottomPadding}
                                                                                            Widget:=overlay:
                                                                                                Slots:=array:
                                                                                                    overlay_slot: 
                                                                                                        Widget:=texture_block{DefaultDesiredSize:=vector2{X:=RColumn6SizeX,Y:=RColumn6SizeY}, DefaultImage:=Images.EmptyTexture}
                                                                                                    overlay_slot: 
                                                                                                        Widget:=P_Row3_Pasive_TextBlock
                                                                                        stack_box_slot: #Column7
                                                                                            Padding:=margin{Left:=InterColumnLeftPadding,Top:=InterColumnTopPadding,Right:=InterColumnRightPadding,Bottom:=InterColumnBottomPadding}
                                                                                            Widget:=overlay:
                                                                                                Slots:=array:
                                                                                                    overlay_slot: 
                                                                                                        Widget:=texture_block{DefaultDesiredSize:=vector2{X:=RColumn7SizeX,Y:=RColumn7SizeY}, DefaultImage:=Images.EmptyTexture}
                                                                                                    overlay_slot: 
                                                                                                        Widget:=P_Row3_PasiveUpgrade_TextBlock
                                                                                        stack_box_slot: #Column8
                                                                                            Padding:=margin{Left:=InterColumnLeftPadding,Top:=InterColumnTopPadding,Right:=InterColumnRightPadding,Bottom:=InterColumnBottomPadding}
                                                                                            Widget:=overlay:
                                                                                                Slots:=array:
                                                                                                    overlay_slot: 
                                                                                                        Widget:=texture_block{DefaultDesiredSize:=vector2{X:=RColumn8SizeX,Y:=RColumn8SizeY}, DefaultImage:=Images.EmptyTexture}
                                                                                                    overlay_slot: 
                                                                                                        HorizontalAlignment:=horizontal_alignment.Left
                                                                                                        VerticalAlignment:=vertical_alignment.Center 
                                                                                                        Padding:=margin{Left:=TextIcon_LeftMargin}
                                                                                                        Widget:=P_Row3_PasiveUpgradeCost_TextBlock               
                                                                                                    overlay_slot: 
                                                                                                        HorizontalAlignment:=horizontal_alignment.Left
                                                                                                        VerticalAlignment:=vertical_alignment.Center 
                                                                                                        Padding:=margin{Left:=Icon_LeftMargin}
                                                                                                        Widget:=texture_block{DefaultDesiredSize:=vector2{X:=IconSmallSizeX,Y:=IconSmallSizeY}, DefaultImage:=Images.AgedSaphire_Icon}
                                                                                                    overlay_slot:
                                                                                                        Padding:=margin{Left:=InterButtonLeftPadding,Top:=InterButtonTopPadding,Right:=InterButtonRightPadding,Bottom:=InterButtonBottomPadding}
                                                                                                        HorizontalAlignment:=horizontal_alignment.Fill
                                                                                                        VerticalAlignment:=vertical_alignment.Fill  
                                                                                                        Widget:=P_Row3PasiveButton
                                                            stack_box_slot: #Pimento
                                                                HorizontalAlignment:=horizontal_alignment.Fill
                                                                VerticalAlignment:=vertical_alignment.Fill
                                                                Padding:=margin{Left:=RowLeftPadding,Top:=RowTopPadding,Right:=RowRightPadding,Bottom:=RowBottomPadding}
                                                                Widget:=overlay:
                                                                    Slots:=array:
                                                                        overlay_slot: 
                                                                            HorizontalAlignment:=horizontal_alignment.Fill
                                                                            Widget:=texture_block{DefaultDesiredSize:=vector2{X:=0.0,Y:=LineSizeY}, DefaultImage:=Images.Img_LineBackground}
                                                                        overlay_slot: 
                                                                            HorizontalAlignment:=horizontal_alignment.Center
                                                                            VerticalAlignment:=vertical_alignment.Center 
                                                                            Widget:=stack_box:
                                                                                Orientation:=orientation.Horizontal
                                                                                    Slots:=array:                  
                                                                                        stack_box_slot: #Column2
                                                                                            Padding:=margin{Left:=InterColumnLeftPadding,Top:=InterColumnTopPadding,Right:=InterColumnRightPadding,Bottom:=InterColumnBottomPadding}
                                                                                            Widget:=overlay:
                                                                                                Slots:=array:
                                                                                                    overlay_slot: 
                                                                                                        Widget:=texture_block{DefaultDesiredSize:=vector2{X:=RColumn2SizeX,Y:=RColumn2SizeY}, DefaultImage:=Images.EmptyTexture}
                                                                                                    overlay_slot: 
                                                                                                        Widget:=texture_block{DefaultDesiredSize:=vector2{X:=IconSizeX,Y:=IconSizeY}, DefaultImage:=Images.Pimento_Icon}
                                                                                        stack_box_slot: #Column3
                                                                                            Padding:=margin{Left:=InterColumnLeftPadding,Top:=InterColumnTopPadding,Right:=InterColumnRightPadding,Bottom:=InterColumnBottomPadding}
                                                                                            Widget:=overlay:
                                                                                                Slots:=array:
                                                                                                    overlay_slot: 
                                                                                                        Widget:=texture_block{DefaultDesiredSize:=vector2{X:=RColumn3SizeX,Y:=RColumn3SizeY}, DefaultImage:=Images.EmptyTexture}
                                                                                                    overlay_slot: 
                                                                                                        Widget:=R_Row4_Capacity_TextBlock
                                                                                        stack_box_slot: #Column4
                                                                                            Padding:=margin{Left:=InterColumnLeftPadding,Top:=InterColumnTopPadding,Right:=InterColumnRightPadding,Bottom:=InterColumnBottomPadding}
                                                                                            Widget:=overlay:
                                                                                                Slots:=array:
                                                                                                    overlay_slot: 
                                                                                                        Widget:=texture_block{DefaultDesiredSize:=vector2{X:=RColumn4SizeX,Y:=RColumn4SizeY}, DefaultImage:=Images.EmptyTexture}
                                                                                                    overlay_slot: 
                                                                                                        Widget:=R_Row4_CapacityUpgrade_TextBlock
                                                                                        stack_box_slot: #Column5
                                                                                            Padding:=margin{Left:=InterColumnLeftPadding,Top:=InterColumnTopPadding,Right:=InterColumnRightPadding,Bottom:=InterColumnBottomPadding}
                                                                                            Widget:=overlay:
                                                                                                Slots:=array:
                                                                                                    overlay_slot: 
                                                                                                        Widget:=texture_block{DefaultDesiredSize:=vector2{X:=RColumn5SizeX,Y:=RColumn5SizeY}, DefaultImage:=Images.EmptyTexture} 
                                                                                                    overlay_slot: 
                                                                                                        HorizontalAlignment:=horizontal_alignment.Left
                                                                                                        VerticalAlignment:=vertical_alignment.Center 
                                                                                                        Padding:=margin{Left:=TextIcon_LeftMargin}
                                                                                                        Widget:=R_Row4_CapacityUpgradeCost_TextBlock               
                                                                                                    overlay_slot: 
                                                                                                        HorizontalAlignment:=horizontal_alignment.Left
                                                                                                        VerticalAlignment:=vertical_alignment.Center 
                                                                                                        Padding:=margin{Left:=Icon_LeftMargin}
                                                                                                        Widget:=texture_block{DefaultDesiredSize:=vector2{X:=IconSmallSizeX,Y:=IconSmallSizeY}, DefaultImage:=Images.Pimento_Icon}
                                                                                                    overlay_slot:
                                                                                                        Padding:=margin{Left:=InterButtonLeftPadding,Top:=InterButtonTopPadding,Right:=InterButtonRightPadding,Bottom:=InterButtonBottomPadding}
                                                                                                        HorizontalAlignment:=horizontal_alignment.Fill
                                                                                                        VerticalAlignment:=vertical_alignment.Fill 
                                                                                                        Widget:=R_Row4CapacityButton
                                                                                        stack_box_slot: #Column6
                                                                                            Padding:=margin{Left:=InterColumnLeftPadding,Top:=InterColumnTopPadding,Right:=InterColumnRightPadding,Bottom:=InterColumnBottomPadding}
                                                                                            Widget:=overlay:
                                                                                                Slots:=array:
                                                                                                    overlay_slot: 
                                                                                                        Widget:=texture_block{DefaultDesiredSize:=vector2{X:=RColumn6SizeX,Y:=RColumn6SizeY}, DefaultImage:=Images.EmptyTexture}
                                                                                                    overlay_slot: 
                                                                                                        Widget:=P_Row4_Pasive_TextBlock
                                                                                        stack_box_slot: #Column7
                                                                                            Padding:=margin{Left:=InterColumnLeftPadding,Top:=InterColumnTopPadding,Right:=InterColumnRightPadding,Bottom:=InterColumnBottomPadding}
                                                                                            Widget:=overlay:
                                                                                                Slots:=array:
                                                                                                    overlay_slot: 
                                                                                                        Widget:=texture_block{DefaultDesiredSize:=vector2{X:=RColumn7SizeX,Y:=RColumn7SizeY}, DefaultImage:=Images.EmptyTexture}
                                                                                                    overlay_slot: 
                                                                                                        Widget:=P_Row4_PasiveUpgrade_TextBlock
                                                                                        stack_box_slot: #Column8
                                                                                            Padding:=margin{Left:=InterColumnLeftPadding,Top:=InterColumnTopPadding,Right:=InterColumnRightPadding,Bottom:=InterColumnBottomPadding}
                                                                                            Widget:=overlay:
                                                                                                Slots:=array:
                                                                                                    overlay_slot: 
                                                                                                        Widget:=texture_block{DefaultDesiredSize:=vector2{X:=RColumn8SizeX,Y:=RColumn8SizeY}, DefaultImage:=Images.EmptyTexture}
                                                                                                    overlay_slot: 
                                                                                                        HorizontalAlignment:=horizontal_alignment.Left
                                                                                                        VerticalAlignment:=vertical_alignment.Center 
                                                                                                        Padding:=margin{Left:=TextIcon_LeftMargin}
                                                                                                        Widget:=P_Row4_PasiveUpgradeCost_TextBlock               
                                                                                                    overlay_slot: 
                                                                                                        HorizontalAlignment:=horizontal_alignment.Left
                                                                                                        VerticalAlignment:=vertical_alignment.Center 
                                                                                                        Padding:=margin{Left:=Icon_LeftMargin}
                                                                                                        Widget:=texture_block{DefaultDesiredSize:=vector2{X:=IconSmallSizeX,Y:=IconSmallSizeY}, DefaultImage:=Images.AgedSaphire_Icon}
                                                                                                    overlay_slot:
                                                                                                        Padding:=margin{Left:=InterButtonLeftPadding,Top:=InterButtonTopPadding,Right:=InterButtonRightPadding,Bottom:=InterButtonBottomPadding}
                                                                                                        HorizontalAlignment:=horizontal_alignment.Fill
                                                                                                        VerticalAlignment:=vertical_alignment.Fill  
                                                                                                        Widget:=P_Row4PasiveButton

                                                            stack_box_slot: #Line explanation
                                                                Padding:=margin{Left:=RowLeftPadding,Top:=RowTopPadding,Right:=RowRightPadding,Bottom:=RowBottomPadding}
                                                                    Widget:=overlay:
                                                                        Slots:=array:
                                                                            overlay_slot: 
                                                                                HorizontalAlignment:=horizontal_alignment.Center
                                                                                VerticalAlignment:=vertical_alignment.Center
                                                                                Widget:=stack_box:
                                                                                    Orientation:=orientation.Horizontal
                                                                                        Slots:=array:
                                                                                            stack_box_slot: 
                                                                                                Padding:=margin{Left:=InterColumnLeftPadding,Top:=InterColumnTopPadding,Right:=InterColumnRightPadding,Bottom:=InterColumnBottomPadding}
                                                                                                Widget:=overlay:
                                                                                                    Slots:=array:
                                                                                                        overlay_slot: 
                                                                                                            Widget:=texture_block{DefaultDesiredSize:=vector2{X:=RColumn8SizeX,Y:=RColumn8SizeY}, DefaultImage:=Images.EmptyTexture}
                                                                                                        overlay_slot: 
                                                                                                            HorizontalAlignment:=horizontal_alignment.Right
                                                                                                            VerticalAlignment:=vertical_alignment.Center
                                                                                                            Padding:=margin{Left:=InterColumnLeftPadding,Top:=InterColumnTopPadding,Right:=InterColumnRightPadding,Bottom:=InterColumnBottomPadding}
                                                                                                            Widget:=text_block{DefaultText:=BasicMessage("*Revenue per second based on your pickaxe's damage and efficiency per resource."), DefaultTextColor:=White, DefaultShadowOffset:=ShadowOffset, DefaultShadowOpacity := DefaultTextShadowOpacity}

                                                            stack_box_slot: #Line 8: All titles
                                                                Padding:=margin{Left:=RowLeftPadding,Top:=RowTopPadding,Right:=RowRightPadding,Bottom:=RowBottomPadding}
                                                                    Widget:=overlay:
                                                                        Slots:=array:
                                                                            overlay_slot: 
                                                                                HorizontalAlignment:=horizontal_alignment.Center
                                                                                VerticalAlignment:=vertical_alignment.Center
                                                                                Widget:=stack_box:
                                                                                    Orientation:=orientation.Horizontal
                                                                                        Slots:=array:
                                                                                            stack_box_slot: 
                                                                                                Padding:=margin{Left:=InterColumnLeftPadding,Top:=InterColumnTopPadding,Right:=InterColumnRightPadding,Bottom:=InterColumnBottomPadding}
                                                                                                Widget:=overlay:
                                                                                                    Slots:=array:
                                                                                                        overlay_slot: 
                                                                                                            Widget:=texture_block{DefaultDesiredSize:=vector2{X:=BigCloseButtonSizeX,Y:=BigCloseButtonSizeY},DefaultImage:=Images.EmptyTexture}
                                                                                                        overlay_slot: 
                                                                                                            HorizontalAlignment:=horizontal_alignment.Fill
                                                                                                            VerticalAlignment:=vertical_alignment.Fill
                                                                                                            Padding:=margin{Left:=InterColumnLeftPadding,Top:=InterColumnTopPadding,Right:=InterColumnRightPadding,Bottom:=InterColumnBottomPadding}
                                                                                                            Widget:=CloseButton
                                                                        

MakeDepositsUpgradeUI<constructor>(
    InBakeryMachine : bakery_machines_manager)<transacts> := deposits_upgrades_ui :
        BakeryMachinesManager := InBakeryMachine
                                                                                                             