using { /UnrealEngine.com/Temporary/UI }
using { /UnrealEngine.com/Temporary/SpatialMath }
using { /Verse.org/Simulation }
using { /Fortnite.com/Devices }
using { /Fortnite.com/UI }
using { /Verse.org/Colors/NamedColors }
using { /Verse.org/Colors }

# See https://dev.epicgames.com/documentation/en-us/uefn/create-your-own-device-in-verse for how to create a verse device.

# A Verse-authored creative device that can be placed in a level
drone_upgrades_ui := class():

    BakeryMachinesManager : bakery_machines_manager
    
    BackgroundImageSizeX : float = 1375.0
    BackgroundImageSizeY : float = 1100.0

    #LineSizeX : float = 1000.0
    #LineSizeY : float = 80.0 #Size of button regular

    PTopColumn1SizeX : float = 350.0
    PTopColumn1SizeY : float = 80.0
    PTopColumn2SizeX : float = 135.0
    PTopColumn2SizeY : float = 80.0

    DPColumn1SizeX : float = 80.0
    DPColumn1SizeY : float = 80.0
    DPColumn2SizeX : float = 475.0
    DPColumn2SizeY : float = 80.0
    DPColumn3SizeX : float = 150.0
    DPColumn3SizeY : float = 80.0
    DPColumn4SizeX : float = 200.0
    DPColumn4SizeY : float = 80.0
    DPColumn5SizeX : float = 200.0
    DPColumn5SizeY : float = 80.0
    DPColumn6SizeX : float = 225.0
    DPColumn6SizeY : float = 80.0

    TitleTextBlock : text_block = text_block{DefaultTextColor:=White, DefaultShadowOffset:=ShadowOffset, DefaultShadowOpacity := DefaultTextShadowOpacity}

    Sapphires_Texture : texture_block = texture_block{DefaultDesiredSize:=vector2{X:=IconSmallSizeX,Y:=IconSmallSizeY}, DefaultImage:=Images.AgedSaphire_Icon}
    TotalSapphires_BackgroundTexture : texture_block = texture_block{DefaultDesiredSize:=vector2{X:=SizeBoxSaphiresX,Y:=SizeBoxSaphiresY}, DefaultImage:=Images.Img_LineBackground2}
    TotalSapphires_TextBlock : text_block = text_block{DefaultTextColor:=AgedSapphiresTextColor,DefaultShadowOffset:=ShadowOffset, DefaultShadowOpacity := DefaultTextShadowOpacity}
    TotalAgedSapphires_TextBlock : text_block = text_block{DefaultTextColor:=AgedSapphiresTextColor,DefaultShadowOffset:=ShadowOffset, DefaultShadowOpacity := DefaultTextShadowOpacity}

    SFP_TopRow1_TextBlock : text_block = text_block{DefaultTextColor:=White, DefaultShadowOffset:=ShadowOffset, DefaultShadowOpacity := DefaultTextShadowOpacity}
    SFP_TopRow2_TextBlock : text_block = text_block{DefaultTextColor:=White, DefaultShadowOffset:=ShadowOffset, DefaultShadowOpacity := DefaultTextShadowOpacity}
    SFP_TopRow3_TextBlock : text_block = text_block{DefaultTextColor:=White, DefaultShadowOffset:=ShadowOffset, DefaultShadowOpacity := DefaultTextShadowOpacity}

    SFP_SingleB_TextBlock : text_block = text_block{DefaultTextColor:=CurrentTextColor, DefaultShadowOffset:=ShadowOffset, DefaultShadowOpacity := DefaultTextShadowOpacity}
    SFP_SingleCB_TextBlock : text_block = text_block{DefaultTextColor:=CurrentTextColor, DefaultShadowOffset:=ShadowOffset, DefaultShadowOpacity := DefaultTextShadowOpacity}
    SFP_SinglePCS_TextBlock : text_block = text_block{DefaultTextColor:=CurrentTextColor, DefaultShadowOffset:=ShadowOffset, DefaultShadowOpacity := DefaultTextShadowOpacity}

    SFP_Row1_IconTexture : texture_block = texture_block{DefaultDesiredSize:=vector2{X:=SizeSmallIconX,Y:=SizeSmallIconY},DefaultImage:=Images.EmptyTexture}
    SFP_Row2_IconTexture : texture_block = texture_block{DefaultDesiredSize:=vector2{X:=SizeSmallIconX,Y:=SizeSmallIconY},DefaultImage:=Images.EmptyTexture}
    SFP_Row3_IconTexture : texture_block = texture_block{DefaultDesiredSize:=vector2{X:=SizeSmallIconX,Y:=SizeSmallIconY},DefaultImage:=Images.EmptyTexture}
   
    DP_Row1_Current_TextBlock : text_block = text_block{DefaultTextColor:=CurrentTextColor, DefaultShadowOffset:=ShadowOffset, DefaultShadowOpacity := DefaultTextShadowOpacity}
    DP_Row2_Current_TextBlock : text_block = text_block{DefaultTextColor:=CurrentTextColor, DefaultShadowOffset:=ShadowOffset, DefaultShadowOpacity := DefaultTextShadowOpacity}
    DP_Row3_Current_TextBlock : text_block = text_block{DefaultTextColor:=CurrentTextColor, DefaultShadowOffset:=ShadowOffset, DefaultShadowOpacity := DefaultTextShadowOpacity}
    DP_Row4_Current_TextBlock : text_block = text_block{DefaultTextColor:=CurrentTextColor, DefaultShadowOffset:=ShadowOffset, DefaultShadowOpacity := DefaultTextShadowOpacity}
    DP_Row5_Current_TextBlock : text_block = text_block{DefaultTextColor:=CurrentTextColor, DefaultShadowOffset:=ShadowOffset, DefaultShadowOpacity := DefaultTextShadowOpacity}

    DP_Row1_Upgrade_TextBlock : text_block = text_block{DefaultTextColor:=UpgradeTextColor, DefaultShadowOffset:=ShadowOffset, DefaultShadowOpacity := DefaultTextShadowOpacity}
    DP_Row2_Upgrade_TextBlock : text_block = text_block{DefaultTextColor:=UpgradeTextColor, DefaultShadowOffset:=ShadowOffset, DefaultShadowOpacity := DefaultTextShadowOpacity}
    DP_Row3_Upgrade_TextBlock : text_block = text_block{DefaultTextColor:=UpgradeTextColor, DefaultShadowOffset:=ShadowOffset, DefaultShadowOpacity := DefaultTextShadowOpacity}
    DP_Row4_Upgrade_TextBlock : text_block = text_block{DefaultTextColor:=UpgradeTextColor, DefaultShadowOffset:=ShadowOffset, DefaultShadowOpacity := DefaultTextShadowOpacity}
    DP_Row5_Upgrade_TextBlock : text_block = text_block{DefaultTextColor:=UpgradeTextColor, DefaultShadowOffset:=ShadowOffset, DefaultShadowOpacity := DefaultTextShadowOpacity}

    DP_Row1_UpgradeCost_TextBlock : text_block = text_block{DefaultTextColor:=SapphiresTextColor, DefaultShadowOffset:=ShadowOffset, DefaultShadowOpacity := DefaultTextShadowOpacity}
    DP_Row2_UpgradeCost_TextBlock : text_block = text_block{DefaultTextColor:=AgedSapphiresTextColor, DefaultShadowOffset:=ShadowOffset, DefaultShadowOpacity := DefaultTextShadowOpacity}
    DP_Row3_UpgradeCost_TextBlock : text_block = text_block{DefaultTextColor:=AgedSapphiresTextColor, DefaultShadowOffset:=ShadowOffset, DefaultShadowOpacity := DefaultTextShadowOpacity}
    DP_Row4_UpgradeCost_TextBlock : text_block = text_block{DefaultTextColor:=AgedSapphiresTextColor, DefaultShadowOffset:=ShadowOffset, DefaultShadowOpacity := DefaultTextShadowOpacity}
    #DP_Row5_UpgradeCost_TextBlock : text_block = text_block{DefaultTextColor:=SapphiresTextColor, DefaultShadowOffset:=ShadowOffset, DefaultShadowOpacity := DefaultTextShadowOpacity}

    DP_Row1Button : button_quiet = button_quiet{}
    DP_Row2Button : button_quiet = button_quiet{}
    DP_Row3Button : button_quiet = button_quiet{}
    DP_Row4Button : button_quiet = button_quiet{}
    #DP_Row5Button : button_quiet = button_quiet{}

    var MaybeCanvas : ?canvas = false
    var UIIsDisplayed : logic = false

    var TimeElapsed : int = 0

    #Arrays
    var AllSFP_TopRows_TextBlock : []text_block = array{}
    var AllSFP_Resourcers_TextBlock : []text_block = array{}
    var AllSFP_Rows_IconTexture : []texture_block = array{}

    var AllDP_Rows_Current_TextBlock : []text_block = array{}
    var AllDP_Rows_Upgrade_TextBlock : []text_block = array{}
    var AllDP_Rows_UpgradeCost_TextBlock : []text_block = array{}
    var AllDP_RowButtons : []button_quiet = array{}

    #Close
    CloseButton : button_loud = button_loud{}


    InitUI() : void =
        SetArrays()
        InitSubscribers()

        Canvas := MakeCanvas()
        set MaybeCanvas = option{Canvas}

        SFP_Row1_IconTexture.SetImage(Images.Sapphire_Icon)
        SFP_Row2_IconTexture.SetImage(Images.Sapphire_Icon)
        SFP_Row3_IconTexture.SetImage(Images.Sapphire_Icon)

        SFP_TopRow1_TextBlock.SetText(BasicMessage("/ Bread: "))
        SFP_TopRow2_TextBlock.SetText(BasicMessage("/ Cheese Bread: "))
        SFP_TopRow3_TextBlock.SetText(BasicMessage("/ Sandwich: "))

    SetArrays() : void =

        set AllSFP_TopRows_TextBlock = array{SFP_TopRow1_TextBlock,SFP_TopRow2_TextBlock,SFP_TopRow3_TextBlock}
        set AllSFP_Resourcers_TextBlock = AllSFP_Resourcers_TextBlock
        set AllSFP_Rows_IconTexture = array{SFP_Row1_IconTexture,SFP_Row2_IconTexture,SFP_Row3_IconTexture}
        
        set AllDP_Rows_Current_TextBlock = array{DP_Row1_Current_TextBlock,DP_Row2_Current_TextBlock,DP_Row3_Current_TextBlock,DP_Row4_Current_TextBlock}
        set AllDP_Rows_Upgrade_TextBlock = array{DP_Row1_Upgrade_TextBlock,DP_Row2_Upgrade_TextBlock,DP_Row3_Upgrade_TextBlock,DP_Row4_Upgrade_TextBlock}
        set AllDP_Rows_UpgradeCost_TextBlock = array{DP_Row1_UpgradeCost_TextBlock,DP_Row2_UpgradeCost_TextBlock,DP_Row3_UpgradeCost_TextBlock,DP_Row4_UpgradeCost_TextBlock}

        set AllDP_RowButtons = array{DP_Row1Button,DP_Row2Button,DP_Row3Button,DP_Row4Button} 

    InitSubscribers() : void =

        #Drone speed
        DP_Row1Button.OnClick().Subscribe(handler_upgrades{MaybeBakeryMachinesManager := option{BakeryMachinesManager}}.HandlerDroneSpeedUpgrade)  

        #Drone capacity
        DP_Row2Button.OnClick().Subscribe(handler_upgrades{MaybeBakeryMachinesManager := option{BakeryMachinesManager}, BakerType := BB_TYPE}.HandlerSalesEfficiencyUpgrade)
        DP_Row3Button.OnClick().Subscribe(handler_upgrades{MaybeBakeryMachinesManager := option{BakeryMachinesManager}, BakerType := CBB_TYPE}.HandlerSalesEfficiencyUpgrade) 
        DP_Row4Button.OnClick().Subscribe(handler_upgrades{MaybeBakeryMachinesManager := option{BakeryMachinesManager}, BakerType := PCSB_TYPE}.HandlerSalesEfficiencyUpgrade)  
 
        CloseButton.OnClick().Subscribe(Remove_Canvas) 


    UpdateUI(PlayerData : player_data_table) : void =

        CloseButton.SetText(BasicMessage("CLOSE"))

        #Inventory resourcers
        TotalSapphires_TextBlock.SetText(BasicMessage(GetMagnitudeLetter(PlayerData.Inventory_Sapphires,"", "")))
        TotalAgedSapphires_TextBlock.SetText(BasicMessage(GetMagnitudeLetter(PlayerData.Inventory_AgedSapphires,"", "")))

        if:
            B_LVEfficiency:=PlayerData.AllLV_SalesEfficiency[0]
            CB_LVEfficiency:=PlayerData.AllLV_SalesEfficiency[1]
            PCS_LVEfficiency:=PlayerData.AllLV_SalesEfficiency[2]
        then:
            BTuple:= CalculeSalesEfficiency(B_LVEfficiency)
            CBTuple:= CalculeSalesEfficiency(CB_LVEfficiency)
            PCSTuple:= CalculeSalesEfficiency(PCS_LVEfficiency)

            SapphiresPerB := (1.0+(BTuple(0)/100.0))*BakeryMachinesManager.Base_SapphiresPerB
            SapphiresPerCB := (1.0+(CBTuple(0)/100.0))*BakeryMachinesManager.Base_SapphiresPerCB
            SapphiresPerPCS := (1.0+(PCSTuple(0)/100.0))*BakeryMachinesManager.Base_SapphiresPerPCS

            B_CurrentSalesEfficiency:=BTuple(0)
            B_UpgradeCostSalesEfficiency:=BTuple(1)
            B_UpgradeSalesEfficiency:=BTuple(2)

            CB_CurrentSalesEfficiency:=CBTuple(0)
            CB_UpgradeCostSalesEfficiency:=CBTuple(1)
            CB_UpgradeSalesEfficiency:=CBTuple(2)

            PCS_CurrentSalesEfficiency:=PCSTuple(0)
            PCS_UpgradeCostSalesEfficiency:=PCSTuple(1)
            PCS_UpgradeSalesEfficiency:=PCSTuple(2)

            SFP_SingleB_TextBlock.SetText(BasicMessage("{GetOptimizedMagnitudeLetter(SapphiresPerB)}"))
            SFP_SingleCB_TextBlock.SetText(BasicMessage("{GetOptimizedMagnitudeLetter(SapphiresPerCB)}"))
            SFP_SinglePCS_TextBlock.SetText(BasicMessage("{GetOptimizedMagnitudeLetter(SapphiresPerPCS)}"))

            AllProductTuples := array{BTuple,CBTuple,PCSTuple}

            #Speed
            SpeedTuple:=CalculeDroneSpeed(PlayerData.LV_DroneSpeed)
            MaxSpeedLV := CalculeSpeedLimitLevel(Base_Drone_Speed*1.0)
            if:
                Speed_Rows_Current_TextBlock := AllDP_Rows_Current_TextBlock[0]
                Speed_Rows_Upgrade_TextBlock := AllDP_Rows_Upgrade_TextBlock[0]
                Speed_Rows_UpgradeCost_TextBlock := AllDP_Rows_UpgradeCost_TextBlock[0]
            then:
                if(PlayerData.LV_DroneSpeed<MaxSpeedLV):
                    Speed_Rows_Current_TextBlock.SetText(BasicMessage("{SpeedTuple(0)} s"))
                    Speed_Rows_UpgradeCost_TextBlock.SetText(BasicMessage(GetMagnitudeLetter(SpeedTuple(1)," x","")))
                    Speed_Rows_Upgrade_TextBlock.SetText(BasicMessage("-{SpeedTuple(2)} s"))
                    Speed_Rows_Current_TextBlock.SetTextColor(CurrentTextColor)
                    Speed_Rows_Upgrade_TextBlock.SetTextColor(UpgradeTextColor)
                    if(PlayerData.Inventory_AgedSapphires<SpeedTuple(1)):
                        Speed_Rows_UpgradeCost_TextBlock.SetTextColor(NotEnoughTextColor)
                    else:
                        Speed_Rows_UpgradeCost_TextBlock.SetTextColor(AgedSapphiresColor)
                else:
                    Speed_Rows_Current_TextBlock.SetText(BasicMessage("{SpeedTuple(0)} s"))
                    Speed_Rows_UpgradeCost_TextBlock.SetText(BasicMessage("MAXED"))
                    Speed_Rows_UpgradeCost_TextBlock.SetTextColor(MaxedColor)
                    Speed_Rows_Upgrade_TextBlock.SetText(BasicMessage("MAXED"))
                    Speed_Rows_Upgrade_TextBlock.SetTextColor(MaxedColor)

            #Sales efficiency
            for(Product:=0..2):
                if:
                    LVEfficiency := PlayerData.AllLV_SalesEfficiency[Product]
                    TupleEfficiency := AllProductTuples[Product]
                    Product_DP_Rows_Current_TextBlock := AllDP_Rows_Current_TextBlock[Product+1]
                    Product_DP_Rows_Upgrade_TextBlock := AllDP_Rows_Upgrade_TextBlock[Product+1]
                    Product_DP_Rows_UpgradeCost_TextBlock := AllDP_Rows_UpgradeCost_TextBlock[Product+1]
                then:
                    Product_CurrentSalesEfficiency:=TupleEfficiency(0)
                    Product_UpgradeCostSalesEfficiency:=TupleEfficiency(1)
                    Product_UpgradeSalesEfficiency:=TupleEfficiency(2)
                    if(LVEfficiency<MaxSalesLV):
                        Product_DP_Rows_Current_TextBlock.SetText(BasicMessage(GetMagnitudeLetter(TupleEfficiency(0)+100.0,""," %")))
                        Product_DP_Rows_UpgradeCost_TextBlock.SetText(BasicMessage(GetMagnitudeLetter(TupleEfficiency(1),"x ","")))
                        Product_DP_Rows_Upgrade_TextBlock.SetText(BasicMessage(GetMagnitudeLetter(TupleEfficiency(2),"+"," %")))
                        Product_DP_Rows_Current_TextBlock.SetTextColor(CurrentTextColor)
                        Product_DP_Rows_Upgrade_TextBlock.SetTextColor(UpgradeTextColor)
                        if(PlayerData.Inventory_AgedSapphires<TupleEfficiency(1)):
                            Product_DP_Rows_UpgradeCost_TextBlock.SetTextColor(NotEnoughTextColor)
                        else:
                            Product_DP_Rows_UpgradeCost_TextBlock.SetTextColor(AgedSapphiresTextColor)
                    else:
                        Product_DP_Rows_Current_TextBlock.SetText(BasicMessage(GetMagnitudeLetter(TupleEfficiency(0)+100.0,""," %")))
                        Product_DP_Rows_UpgradeCost_TextBlock.SetText(BasicMessage("MAXED"))
                        Product_DP_Rows_UpgradeCost_TextBlock.SetTextColor(MaxedColor)
                        Product_DP_Rows_Upgrade_TextBlock.SetText(BasicMessage("MAXED"))
                        Product_DP_Rows_Upgrade_TextBlock.SetTextColor(MaxedColor)

    Remove_Canvas(Message : widget_message) : void =
        set UIIsDisplayed = false
        if:
            Agent:=BakeryMachinesManager.MaybeAgentOwner?
            Player:=player[Agent]
            PlayerUI:=GetPlayerUI[Player]
            Canvas := MaybeCanvas?
        then:
            PlayerUI.RemoveWidget(Canvas)
            RemoveDelay_OnRemove()

    Remove_CanvasWithoutMessage() : void =
        set UIIsDisplayed = false
        #Print("Trying to remove UI")
        if:
            Agent:=BakeryMachinesManager.MaybeAgentOwner?
            Player:=player[Agent]
            PlayerUI:=GetPlayerUI[Player]
            Canvas := MaybeCanvas?
        then:
            #Print("Removing UI")
            PlayerUI.RemoveWidget(Canvas)
            RemoveDelay_OnRemove()

    RemoveDelay_OnAdd() : void =
        TitleTextBlock.SetText(BasicMessage("Drone perks and product sales"))
        #CurrentBuyOption_textblock.SetVisibility(widget_visibility.Hidden)

    RemoveDelay_OnRemove() : void =
        TitleTextBlock.SetText(BasicMessage("Drone perks and product sales "))
        #CurrentBuyOption_textblock.SetVisibility(widget_visibility.Hidden)

            
    Display_Canvas() : void =
        set TimeElapsed = 0
        set UIIsDisplayed = true
        spawn: 
            StartTimeElapsed()
        if:
            Agent:=BakeryMachinesManager.MaybeAgentOwner?
            PlayerUI:=GetPlayerUI[player[Agent]]
            Canvas := MaybeCanvas?
        then:
            ##Print("Assigning created UI before")
            PlayerUI.AddWidget(Canvas,player_ui_slot{InputMode := ui_input_mode.All, ZOrder:=1}) 
            RemoveDelay_OnAdd() 

    StartTimeElapsed()<suspends> : void =
        loop:
            ##Print("IN THE LOOP TimeElapsed: {TimeElapsed}")
            if(TimeElapsed > MaxUIDisplayTime):
                ##Print("Closing the UI because 20 seconds has pased")
                Remove_CanvasWithoutMessage()
                break
            if(UIIsDisplayed = false):
                ##Print("Breaking Time Elapsed loop")
                break
            Sleep(1.0)
            set TimeElapsed = TimeElapsed + 1

    MakeCanvas(): canvas=
        Canvas := canvas:
            Slots:=array:
                #Black Background
                canvas_slot:
                    Anchors:=anchors{Minimum:=vector2{X:=0.5,Y:=0.5},Maximum:=vector2{X:=0.5,Y:=0.5}}
                    ZOrder:=0
                    Alignment:=vector2{X:=0.5,Y:=0.5}
                    Widget:=texture_block:
                        DefaultImage:=Images.BlackBackground
                        DefaultDesiredSize:=vector2{X:=BackgroundBlackUISizeX,Y:=BackgroundBlackUISizeY}
                #Background image
                canvas_slot:
                    Anchors:=anchors{Minimum:=vector2{X:=0.5,Y:=0.5},Maximum:=vector2{X:=0.5,Y:=0.5}}
                    ZOrder:=0
                    Offsets:=margin{Left:=0.0,Top:=0.0,Right:=BackgroundImageSizeX,Bottom:=BackgroundImageSizeY} #PositionX, PositionY, SizeX, SizeY
                    Alignment:=vector2{X:=0.5,Y:=0.5}
                    Widget:=texture_block:
                        DefaultImage:=Images.Img_MainBackground
                        DefaultDesiredSize:=vector2{X:=BackgroundImageSizeX,Y:=BackgroundImageSizeY}
                #Main overlay
                canvas_slot:
                    Anchors:=anchors{Minimum:=vector2{X:=0.5,Y:=0.5},Maximum:=vector2{X:=0.5,Y:=0.5}}
                    ZOrder:=0
                    #Offsets:=margin{Left:=1920.0,Top:=0.0,Right:=0.0,Bottom:=1080.0} #PositionX, PositionY, SizeX, SizeY
                    Alignment:=vector2{X:=0.5,Y:=0.5}
                    Widget:=overlay:
                        Slots:=array:
                            overlay_slot:
                                HorizontalAlignment:=horizontal_alignment.Center
                                VerticalAlignment:=vertical_alignment.Center
                                Widget:=stack_box:
                                    Orientation:=orientation.Vertical
                                        Slots:=array:
                                            stack_box_slot: #Sapphires per product
                                                Widget:=stack_box:
                                                    Orientation:=orientation.Horizontal
                                                        Slots:=array:
                                                            stack_box_slot: 
                                                                HorizontalAlignment:=horizontal_alignment.Fill
                                                                VerticalAlignment:=vertical_alignment.Fill 
                                                                Widget:=stack_box:
                                                                    Orientation:=orientation.Vertical
                                                                    Slots:=array:
                                                                        stack_box_slot: #Column bread
                                                                            Padding:=margin{Left:=RowLeftPadding,Top:=RowTopPadding,Right:=RowRightPadding,Bottom:=RowBottomPadding}
                                                                            HorizontalAlignment:=horizontal_alignment.Center
                                                                            VerticalAlignment:=vertical_alignment.Top
                                                                            Widget:=overlay: 
                                                                                Slots:=array:
                                                                                    overlay_slot:
                                                                                        HorizontalAlignment:=horizontal_alignment.Fill
                                                                                        VerticalAlignment:=vertical_alignment.Center
                                                                                        Widget:=texture_block{DefaultDesiredSize :=vector2{X:=0.0,Y:=LineSizeY}, DefaultImage:=Images.Img_LineBackground}
                                                                                    overlay_slot:
                                                                                        HorizontalAlignment:=horizontal_alignment.Center
                                                                                        VerticalAlignment:=vertical_alignment.Center
                                                                                        #Upgrade level stack box
                                                                                        Widget:=stack_box: 
                                                                                            Orientation:=orientation.Horizontal
                                                                                            Slots:=array:    
                                                                                                stack_box_slot: 
                                                                                                    Padding:=margin{Left:=InterColumnLeftPadding,Top:=InterColumnTopPadding,Right:=InterColumnRightPadding,Bottom:=InterColumnBottomPadding}
                                                                                                    HorizontalAlignment:=horizontal_alignment.Center
                                                                                                    VerticalAlignment:=vertical_alignment.Center
                                                                                                    Widget:=overlay:
                                                                                                        Slots:=array:
                                                                                                            overlay_slot:
                                                                                                                Widget:=texture_block{DefaultDesiredSize:=vector2{X:=PTopColumn1SizeX,Y:=PTopColumn1SizeY},DefaultImage:=Images.EmptyTexture} 
                                                                                                            overlay_slot: 
                                                                                                                HorizontalAlignment:=horizontal_alignment.Left
                                                                                                                VerticalAlignment:=vertical_alignment.Center 
                                                                                                                Padding:=margin{Left:=65.0,Top:=0.0,Right:=0.0,Bottom:=0.0}
                                                                                                                Widget:=SFP_TopRow1_TextBlock 
                                                                                                            overlay_slot: 
                                                                                                                HorizontalAlignment:=horizontal_alignment.Left
                                                                                                                VerticalAlignment:=vertical_alignment.Center 
                                                                                                                Padding:=margin{Left:=10.0,Top:=0.0,Right:=0.0,Bottom:=0.0}
                                                                                                                Widget:=SFP_Row1_IconTexture
                                                                                                stack_box_slot: #Column 2
                                                                                                    Padding:=margin{Left:=InterColumnLeftPadding,Top:=InterColumnTopPadding,Right:=InterColumnRightPadding,Bottom:=InterColumnBottomPadding}
                                                                                                    HorizontalAlignment:=horizontal_alignment.Center
                                                                                                    VerticalAlignment:=vertical_alignment.Center
                                                                                                    Widget:=overlay:
                                                                                                        Slots:=array:
                                                                                                            overlay_slot:
                                                                                                                Widget:=texture_block{DefaultDesiredSize:=vector2{X:=PTopColumn2SizeX,Y:=PTopColumn2SizeY},DefaultImage:=Images.EmptyTexture}
                                                                                                            overlay_slot: 
                                                                                                                HorizontalAlignment:=horizontal_alignment.Left
                                                                                                                VerticalAlignment:=vertical_alignment.Center 
                                                                                                                Widget:=SFP_SingleB_TextBlock 
                
                                                                        stack_box_slot: #Column Cheese bread
                                                                            Padding:=margin{Left:=RowLeftPadding,Top:=RowTopPadding,Right:=RowRightPadding,Bottom:=RowBottomPadding}
                                                                            HorizontalAlignment:=horizontal_alignment.Center
                                                                            VerticalAlignment:=vertical_alignment.Top
                                                                            Widget:=overlay: 
                                                                                Slots:=array:
                                                                                    overlay_slot:
                                                                                        HorizontalAlignment:=horizontal_alignment.Fill
                                                                                        VerticalAlignment:=vertical_alignment.Center
                                                                                        Widget:=texture_block{DefaultDesiredSize :=vector2{X:=0.0,Y:=LineSizeY}, DefaultImage:=Images.Img_LineBackground}
                                                                                    overlay_slot:
                                                                                        HorizontalAlignment:=horizontal_alignment.Center
                                                                                        VerticalAlignment:=vertical_alignment.Center
                                                                                        #Upgrade level stack box
                                                                                        Widget:=stack_box: 
                                                                                            Orientation:=orientation.Horizontal
                                                                                            Slots:=array:    
                                                                                                stack_box_slot: #Column 1
                                                                                                    Padding:=margin{Left:=InterColumnLeftPadding,Top:=InterColumnTopPadding,Right:=InterColumnRightPadding,Bottom:=InterColumnBottomPadding}
                                                                                                    HorizontalAlignment:=horizontal_alignment.Center
                                                                                                    VerticalAlignment:=vertical_alignment.Center
                                                                                                    Widget:=overlay:
                                                                                                        Slots:=array:
                                                                                                            overlay_slot:
                                                                                                                Widget:=texture_block{DefaultDesiredSize:=vector2{X:=PTopColumn1SizeX,Y:=PTopColumn1SizeY},DefaultImage:=Images.EmptyTexture}
                                                                                                            overlay_slot: 
                                                                                                                HorizontalAlignment:=horizontal_alignment.Left
                                                                                                                VerticalAlignment:=vertical_alignment.Center 
                                                                                                                Padding:=margin{Left:=65.0,Top:=0.0,Right:=0.0,Bottom:=0.0}
                                                                                                                Widget:=SFP_TopRow2_TextBlock 
                                                                                                            overlay_slot: 
                                                                                                                HorizontalAlignment:=horizontal_alignment.Left
                                                                                                                VerticalAlignment:=vertical_alignment.Center 
                                                                                                                Padding:=margin{Left:=10.0,Top:=0.0,Right:=0.0,Bottom:=0.0}
                                                                                                                Widget:=SFP_Row2_IconTexture
                                                                                                stack_box_slot: #Column 2
                                                                                                    Padding:=margin{Left:=InterColumnLeftPadding,Top:=InterColumnTopPadding,Right:=InterColumnRightPadding,Bottom:=InterColumnBottomPadding}
                                                                                                    HorizontalAlignment:=horizontal_alignment.Center
                                                                                                    VerticalAlignment:=vertical_alignment.Center
                                                                                                    Widget:=overlay:
                                                                                                        Slots:=array:
                                                                                                            overlay_slot:
                                                                                                                Widget:=texture_block{DefaultDesiredSize:=vector2{X:=PTopColumn2SizeX,Y:=PTopColumn2SizeY},DefaultImage:=Images.EmptyTexture}
                                                                                                            overlay_slot: 
                                                                                                                HorizontalAlignment:=horizontal_alignment.Left
                                                                                                                VerticalAlignment:=vertical_alignment.Center 
                                                                                                                Widget:=SFP_SingleCB_TextBlock
                                                                                                                
                                                                        stack_box_slot: #Column Sandwich
                                                                            Padding:=margin{Left:=RowLeftPadding,Top:=RowTopPadding,Right:=RowRightPadding,Bottom:=RowBottomPadding}
                                                                            HorizontalAlignment:=horizontal_alignment.Center
                                                                            VerticalAlignment:=vertical_alignment.Top
                                                                            Widget:=overlay: 
                                                                                Slots:=array:
                                                                                    overlay_slot:
                                                                                        HorizontalAlignment:=horizontal_alignment.Fill
                                                                                        VerticalAlignment:=vertical_alignment.Center
                                                                                        Widget:=texture_block{DefaultDesiredSize :=vector2{X:=0.0,Y:=LineSizeY}, DefaultImage:=Images.Img_LineBackground}
                                                                                    overlay_slot:
                                                                                        HorizontalAlignment:=horizontal_alignment.Center
                                                                                        VerticalAlignment:=vertical_alignment.Center
                                                                                        #Upgrade level stack box
                                                                                        Widget:=stack_box: 
                                                                                            Orientation:=orientation.Horizontal
                                                                                            Slots:=array:    
                                                                                                stack_box_slot: #Column 1
                                                                                                    Padding:=margin{Left:=InterColumnLeftPadding,Top:=InterColumnTopPadding,Right:=InterColumnRightPadding,Bottom:=InterColumnBottomPadding}
                                                                                                    HorizontalAlignment:=horizontal_alignment.Center
                                                                                                    VerticalAlignment:=vertical_alignment.Center
                                                                                                    Widget:=overlay:
                                                                                                        Slots:=array:
                                                                                                            overlay_slot:
                                                                                                                Widget:=texture_block{DefaultDesiredSize:=vector2{X:=PTopColumn1SizeX,Y:=PTopColumn1SizeY},DefaultImage:=Images.EmptyTexture}
                                                                                                            overlay_slot: 
                                                                                                                HorizontalAlignment:=horizontal_alignment.Left
                                                                                                                VerticalAlignment:=vertical_alignment.Center 
                                                                                                                Padding:=margin{Left:=65.0,Top:=0.0,Right:=0.0,Bottom:=0.0}
                                                                                                                Widget:=SFP_TopRow3_TextBlock 
                                                                                                            overlay_slot: 
                                                                                                                HorizontalAlignment:=horizontal_alignment.Left
                                                                                                                VerticalAlignment:=vertical_alignment.Center 
                                                                                                                Padding:=margin{Left:=10.0,Top:=0.0,Right:=0.0,Bottom:=0.0}
                                                                                                                Widget:=SFP_Row3_IconTexture
                                                                                                stack_box_slot: #Column 2
                                                                                                    Padding:=margin{Left:=InterColumnLeftPadding,Top:=InterColumnTopPadding,Right:=InterColumnRightPadding,Bottom:=InterColumnBottomPadding}
                                                                                                    HorizontalAlignment:=horizontal_alignment.Center
                                                                                                    VerticalAlignment:=vertical_alignment.Center
                                                                                                    Widget:=overlay:
                                                                                                        Slots:=array:
                                                                                                            overlay_slot:
                                                                                                                Widget:=texture_block{DefaultDesiredSize:=vector2{X:=PTopColumn2SizeX,Y:=PTopColumn2SizeY},DefaultImage:=Images.EmptyTexture}
                                                                                                            overlay_slot: 
                                                                                                                HorizontalAlignment:=horizontal_alignment.Left
                                                                                                                VerticalAlignment:=vertical_alignment.Center 
                                                                                                                Widget:=SFP_SinglePCS_TextBlock
                                                            stack_box_slot: #Empty space
                                                                Widget:=overlay:
                                                                    Slots:=array:
                                                                        overlay_slot:
                                                                            Widget:=texture_block{DefaultDesiredSize:=vector2{X:=530.0,Y:=50.0},DefaultImage:=Images.EmptyTexture}
                                                            stack_box_slot: #Current sapphires
                                                                Widget:=overlay:
                                                                    Slots:=array:
                                                                        overlay_slot:
                                                                            Widget:=TotalSapphires_BackgroundTexture
                                                                        overlay_slot:
                                                                            HorizontalAlignment:=horizontal_alignment.Left
                                                                            VerticalAlignment:=vertical_alignment.Center 
                                                                            Padding:=margin{Left:=10.0,Top:=0.0,Right:=0.0,Bottom:=0.0}
                                                                            Widget:=Sapphires_Texture
                                                                        overlay_slot: 
                                                                            HorizontalAlignment:=horizontal_alignment.Left
                                                                            VerticalAlignment:=vertical_alignment.Center 
                                                                            Padding:=margin{Left:=65.0,Top:=0.0,Right:=0.0,Bottom:=0.0}
                                                                            Widget:=TotalAgedSapphires_TextBlock
                                            stack_box_slot: #Drone perks
                                                Widget:=stack_box:
                                                    Orientation:=orientation.Vertical
                                                        Slots:=array:
                                                            stack_box_slot: #Line 1: Title
                                                                Padding:=margin{Left:=RowLeftPadding,Top:=RowTopPadding,Right:=RowRightPadding,Bottom:=RowBottomPadding}
                                                                HorizontalAlignment:=horizontal_alignment.Fill
                                                                VerticalAlignment:=vertical_alignment.Bottom
                                                                Widget:=overlay:
                                                                    Slots:=array:
                                                                        overlay_slot: 
                                                                            HorizontalAlignment:=horizontal_alignment.Fill
                                                                            VerticalAlignment:=vertical_alignment.Bottom
                                                                            Widget:=texture_block{DefaultDesiredSize:=vector2{X:=HorizontalLineSizeX,Y:=HorizontalLineSizeY},DefaultImage:=Images.Img_HorizontalLine}
                                                                        overlay_slot: 
                                                                            HorizontalAlignment:=horizontal_alignment.Center
                                                                            VerticalAlignment:=vertical_alignment.Center 
                                                                            Widget:=TitleTextBlock
                                                                        overlay_slot: 
                                                                            HorizontalAlignment:=horizontal_alignment.Center
                                                                            VerticalAlignment:=vertical_alignment.Center 
                                                                            Widget:=texture_block{DefaultDesiredSize:=vector2{X:=10.0,Y:=LineSizeY},DefaultImage:=Images.EmptyTexture}
                                                            stack_box_slot: #Line 2: All titles
                                                                Padding:=margin{Left:=RowLeftPadding,Top:=RowTopPadding,Right:=RowRightPadding,Bottom:=RowBottomPadding}
                                                                    Widget:=overlay:
                                                                        Slots:=array:
                                                                            overlay_slot: 
                                                                                HorizontalAlignment:=horizontal_alignment.Fill
                                                                                VerticalAlignment:=vertical_alignment.Center
                                                                                Widget:=texture_block{DefaultDesiredSize:=vector2{X:=0.0,Y:=LineSizeY},DefaultImage:=Images.Img_LineBackground}
                                                                            overlay_slot: 
                                                                                Widget:=stack_box:
                                                                                    Orientation:=orientation.Horizontal
                                                                                        Slots:=array:
                                                                                            stack_box_slot: 
                                                                                                Padding:=margin{Left:=InterColumnLeftPadding,Top:=InterColumnTopPadding,Right:=InterColumnRightPadding,Bottom:=InterColumnBottomPadding}
                                                                                                Widget:=overlay:
                                                                                                    Slots:=array:
                                                                                                        overlay_slot: 
                                                                                                            Widget:=texture_block{DefaultDesiredSize:=vector2{X:=DPColumn1SizeX,Y:=DPColumn1SizeY},DefaultImage:=Images.EmptyTexture}
                                                                                            stack_box_slot: 
                                                                                                Padding:=margin{Left:=InterColumnLeftPadding,Top:=InterColumnTopPadding,Right:=InterColumnRightPadding,Bottom:=InterColumnBottomPadding}
                                                                                                Widget:=overlay:
                                                                                                    Slots:=array:
                                                                                                        overlay_slot:
                                                                                                            Widget:=text_block{DefaultText:=BasicMessage("Perk"), DefaultTextColor:=White, DefaultShadowOffset:=ShadowOffset, DefaultShadowOpacity := DefaultTextShadowOpacity}
                                                                                                        overlay_slot: 
                                                                                                            Widget:=texture_block{DefaultDesiredSize:=vector2{X:=DPColumn2SizeX,Y:=DPColumn2SizeY},DefaultImage:=Images.EmptyTexture}

                                                                                            stack_box_slot: 
                                                                                                Padding:=margin{Left:=InterColumnLeftPadding,Top:=InterColumnTopPadding,Right:=InterColumnRightPadding,Bottom:=InterColumnBottomPadding}
                                                                                                Widget:=overlay:
                                                                                                    Slots:=array:
                                                                                                        overlay_slot:
                                                                                                            Widget:=text_block{DefaultText:=BasicMessage("Current"), DefaultTextColor:=White, DefaultShadowOffset:=ShadowOffset, DefaultShadowOpacity := DefaultTextShadowOpacity}
                                                                                                        overlay_slot: 
                                                                                                            Widget:=texture_block{DefaultDesiredSize:=vector2{X:=DPColumn4SizeX,Y:=DPColumn4SizeY},DefaultImage:=Images.EmptyTexture}
                                                                                            stack_box_slot: 
                                                                                                Padding:=margin{Left:=InterColumnLeftPadding,Top:=InterColumnTopPadding,Right:=InterColumnRightPadding,Bottom:=InterColumnBottomPadding}
                                                                                                Widget:=overlay:
                                                                                                    Slots:=array:
                                                                                                        overlay_slot:
                                                                                                            Widget:=text_block{DefaultText:=BasicMessage("Upgrade"), DefaultTextColor:=White, DefaultShadowOffset:=ShadowOffset, DefaultShadowOpacity := DefaultTextShadowOpacity}
                                                                                                        overlay_slot: 
                                                                                                            Widget:=texture_block{DefaultDesiredSize:=vector2{X:=DPColumn5SizeX,Y:=DPColumn5SizeY},DefaultImage:=Images.EmptyTexture}
                                                                                            stack_box_slot: 
                                                                                                Padding:=margin{Left:=InterColumnLeftPadding,Top:=InterColumnTopPadding,Right:=InterColumnRightPadding,Bottom:=InterColumnBottomPadding}
                                                                                                Widget:=overlay:
                                                                                                    Slots:=array:
                                                                                                        overlay_slot:
                                                                                                            Widget:=text_block{DefaultText:=BasicMessage("Upgrade cost"), DefaultTextColor:=White, DefaultShadowOffset:=ShadowOffset, DefaultShadowOpacity := DefaultTextShadowOpacity}
                                                                                                        overlay_slot: 
                                                                                                            Widget:=texture_block{DefaultDesiredSize:=vector2{X:=DPColumn6SizeX,Y:=DPColumn6SizeY},DefaultImage:=Images.EmptyTexture}

                                                            <#
                                                            stack_box_slot: #Line 3: 
                                                                Padding:=margin{Left:=RowLeftPadding,Top:=RowTopPadding,Right:=RowRightPadding,Bottom:=RowBottomPadding}
                                                                    Widget:=overlay:
                                                                        Slots:=array:
                                                                            overlay_slot: 
                                                                                HorizontalAlignment:=horizontal_alignment.Fill
                                                                                VerticalAlignment:=vertical_alignment.Center
                                                                                Widget:=texture_block{DefaultDesiredSize:=vector2{X:=0.0,Y:=LineSizeY},DefaultImage:=Images.Img_LineBackground}
                                                                            overlay_slot: 
                                                                                Widget:=stack_box:
                                                                                    Orientation:=orientation.Horizontal
                                                                                        Slots:=array:
                                                                                            stack_box_slot: 
                                                                                                Padding:=margin{Left:=InterColumnLeftPadding,Top:=InterColumnTopPadding,Right:=InterColumnRightPadding,Bottom:=InterColumnBottomPadding}
                                                                                                Widget:=overlay:
                                                                                                    Slots:=array:
                                                                                                        overlay_slot: 
                                                                                                            Widget:=texture_block{DefaultDesiredSize:=vector2{X:=DPColumn1SizeX,Y:=DPColumn1SizeY},DefaultImage:=Images.EmptyTexture}
                                                                                                        overlay_slot: 
                                                                                                            Widget:=texture_block{DefaultDesiredSize:=vector2{X:=IconSmallSizeX,Y:=IconSmallSizeY},DefaultImage:=Images.Speed_Icon}
                                                                                            stack_box_slot: 
                                                                                                Padding:=margin{Left:=InterColumnLeftPadding,Top:=InterColumnTopPadding,Right:=InterColumnRightPadding,Bottom:=InterColumnBottomPadding}
                                                                                                Widget:=overlay:
                                                                                                    Slots:=array:
                                                                                                        overlay_slot:
                                                                                                            Widget:=text_block{DefaultText:=BasicMessage("Delivery speed"), DefaultTextColor:=White, DefaultShadowOffset:=ShadowOffset, DefaultShadowOpacity := DefaultTextShadowOpacity}
                                                                                                        overlay_slot: 
                                                                                                            Widget:=texture_block{DefaultDesiredSize:=vector2{X:=DPColumn2SizeX,Y:=DPColumn2SizeY},DefaultImage:=Images.EmptyTexture}
                                                                                            stack_box_slot: 
                                                                                                Padding:=margin{Left:=InterColumnLeftPadding,Top:=InterColumnTopPadding,Right:=InterColumnRightPadding,Bottom:=InterColumnBottomPadding}
                                                                                                Widget:=overlay:
                                                                                                    Slots:=array:
                                                                                                        overlay_slot:
                                                                                                            Widget:=DP_Row1_Current_TextBlock
                                                                                                        overlay_slot: 
                                                                                                            Widget:=texture_block{DefaultDesiredSize:=vector2{X:=DPColumn4SizeX,Y:=DPColumn4SizeY},DefaultImage:=Images.EmptyTexture}
                                                                                            stack_box_slot: 
                                                                                                Padding:=margin{Left:=InterColumnLeftPadding,Top:=InterColumnTopPadding,Right:=InterColumnRightPadding,Bottom:=InterColumnBottomPadding}
                                                                                                Widget:=overlay:
                                                                                                    Slots:=array:
                                                                                                        overlay_slot:
                                                                                                            Widget:=DP_Row1_Upgrade_TextBlock
                                                                                                        overlay_slot: 
                                                                                                            Widget:=texture_block{DefaultDesiredSize:=vector2{X:=DPColumn5SizeX,Y:=DPColumn5SizeY},DefaultImage:=Images.EmptyTexture}
                                                                                            stack_box_slot: 
                                                                                                Padding:=margin{Left:=InterColumnLeftPadding,Top:=InterColumnTopPadding,Right:=InterColumnRightPadding,Bottom:=InterColumnBottomPadding}
                                                                                                Widget:=overlay:
                                                                                                    Slots:=array:
                                                                                                        overlay_slot: 
                                                                                                            Widget:=texture_block{DefaultDesiredSize:=vector2{X:=DPColumn6SizeX,Y:=DPColumn6SizeY},DefaultImage:=Images.EmptyTexture}
                                                                                                        overlay_slot: 
                                                                                                            HorizontalAlignment:=horizontal_alignment.Left
                                                                                                            VerticalAlignment:=vertical_alignment.Center 
                                                                                                            Padding:=margin{Left:=65.0,Top:=0.0,Right:=0.0,Bottom:=0.0}
                                                                                                            Widget:=DP_Row1_UpgradeCost_TextBlock 
                                                                                                        overlay_slot: 
                                                                                                            HorizontalAlignment:=horizontal_alignment.Left
                                                                                                            VerticalAlignment:=vertical_alignment.Center 
                                                                                                            Padding:=margin{Left:=10.0,Top:=0.0,Right:=0.0,Bottom:=0.0}
                                                                                                            Widget:=texture_block{DefaultDesiredSize:=vector2{X:=IconSmallSizeX,Y:=IconSmallSizeY}, DefaultImage:=Images.Sapphire_Icon}
                                                                                                        overlay_slot:
                                                                                                            HorizontalAlignment:=horizontal_alignment.Fill
                                                                                                            VerticalAlignment:=vertical_alignment.Fill   
                                                                                                            Widget:=DP_Row1Button#>
                                                            stack_box_slot: #Line 5: All titles
                                                                Padding:=margin{Left:=RowLeftPadding,Top:=RowTopPadding,Right:=RowRightPadding,Bottom:=RowBottomPadding}
                                                                    Widget:=overlay:
                                                                        Slots:=array:
                                                                            overlay_slot: 
                                                                                HorizontalAlignment:=horizontal_alignment.Fill
                                                                                VerticalAlignment:=vertical_alignment.Center
                                                                                Widget:=texture_block{DefaultDesiredSize:=vector2{X:=0.0,Y:=LineSizeY},DefaultImage:=Images.Img_LineBackground}
                                                                            overlay_slot: 
                                                                                Widget:=stack_box:
                                                                                    Orientation:=orientation.Horizontal
                                                                                        Slots:=array:
                                                                                            stack_box_slot: 
                                                                                                Padding:=margin{Left:=InterColumnLeftPadding,Top:=InterColumnTopPadding,Right:=InterColumnRightPadding,Bottom:=InterColumnBottomPadding}
                                                                                                Widget:=overlay:
                                                                                                    Slots:=array:
                                                                                                        overlay_slot: 
                                                                                                            Widget:=texture_block{DefaultDesiredSize:=vector2{X:=DPColumn1SizeX,Y:=DPColumn1SizeY},DefaultImage:=Images.EmptyTexture}
                                                                                                        overlay_slot: 
                                                                                                            Widget:=texture_block{DefaultDesiredSize:=vector2{X:=IconSmallSizeX,Y:=IconSmallSizeY},DefaultImage:=Images.Bread_Icon}
                                                                                            stack_box_slot: 
                                                                                                Padding:=margin{Left:=InterColumnLeftPadding,Top:=InterColumnTopPadding,Right:=InterColumnRightPadding,Bottom:=InterColumnBottomPadding}
                                                                                                Widget:=overlay:
                                                                                                    Slots:=array:
                                                                                                        overlay_slot:
                                                                                                            Widget:=text_block{DefaultText:=BasicMessage("Bread sales efficiency"), DefaultTextColor:=White, DefaultShadowOffset:=ShadowOffset, DefaultShadowOpacity := DefaultTextShadowOpacity}
                                                                                                        overlay_slot: 
                                                                                                            Widget:=texture_block{DefaultDesiredSize:=vector2{X:=DPColumn2SizeX,Y:=DPColumn2SizeY},DefaultImage:=Images.EmptyTexture}
                                                                                            stack_box_slot: 
                                                                                                Padding:=margin{Left:=InterColumnLeftPadding,Top:=InterColumnTopPadding,Right:=InterColumnRightPadding,Bottom:=InterColumnBottomPadding}
                                                                                                Widget:=overlay:
                                                                                                    Slots:=array:
                                                                                                        overlay_slot:
                                                                                                            Widget:=DP_Row2_Current_TextBlock
                                                                                                        overlay_slot: 
                                                                                                            Widget:=texture_block{DefaultDesiredSize:=vector2{X:=DPColumn4SizeX,Y:=DPColumn4SizeY},DefaultImage:=Images.EmptyTexture}
                                                                                            stack_box_slot: 
                                                                                                Padding:=margin{Left:=InterColumnLeftPadding,Top:=InterColumnTopPadding,Right:=InterColumnRightPadding,Bottom:=InterColumnBottomPadding}
                                                                                                Widget:=overlay:
                                                                                                    Slots:=array:
                                                                                                        overlay_slot:
                                                                                                            Widget:=DP_Row2_Upgrade_TextBlock
                                                                                                        overlay_slot: 
                                                                                                            Widget:=texture_block{DefaultDesiredSize:=vector2{X:=DPColumn5SizeX,Y:=DPColumn5SizeY},DefaultImage:=Images.EmptyTexture}
                                                                                            stack_box_slot: 
                                                                                                Padding:=margin{Left:=InterColumnLeftPadding,Top:=InterColumnTopPadding,Right:=InterColumnRightPadding,Bottom:=InterColumnBottomPadding}
                                                                                                Widget:=overlay:
                                                                                                    Slots:=array:
                                                                                                        overlay_slot: 
                                                                                                            Widget:=texture_block{DefaultDesiredSize:=vector2{X:=DPColumn6SizeX,Y:=DPColumn6SizeY},DefaultImage:=Images.EmptyTexture}
                                                                                                        overlay_slot: 
                                                                                                            HorizontalAlignment:=horizontal_alignment.Left
                                                                                                            VerticalAlignment:=vertical_alignment.Center 
                                                                                                            Padding:=margin{Left:=65.0,Top:=0.0,Right:=0.0,Bottom:=0.0}
                                                                                                            Widget:=DP_Row2_UpgradeCost_TextBlock 
                                                                                                        overlay_slot: 
                                                                                                            HorizontalAlignment:=horizontal_alignment.Left
                                                                                                            VerticalAlignment:=vertical_alignment.Center 
                                                                                                            Padding:=margin{Left:=10.0,Top:=0.0,Right:=0.0,Bottom:=0.0}
                                                                                                            Widget:=texture_block{DefaultDesiredSize:=vector2{X:=IconSmallSizeX,Y:=IconSmallSizeY}, DefaultImage:=Images.AgedSaphire_Icon}
                                                                                                        overlay_slot:
                                                                                                            HorizontalAlignment:=horizontal_alignment.Fill
                                                                                                            VerticalAlignment:=vertical_alignment.Fill   
                                                                                                            Widget:=DP_Row2Button
                                                                                                

                                                            stack_box_slot: #Line 6: All titles
                                                                Padding:=margin{Left:=RowLeftPadding,Top:=RowTopPadding,Right:=RowRightPadding,Bottom:=RowBottomPadding}
                                                                    Widget:=overlay:
                                                                        Slots:=array:
                                                                            overlay_slot: 
                                                                                HorizontalAlignment:=horizontal_alignment.Fill
                                                                                VerticalAlignment:=vertical_alignment.Center
                                                                                Widget:=texture_block{DefaultDesiredSize:=vector2{X:=0.0,Y:=LineSizeY},DefaultImage:=Images.Img_LineBackground}
                                                                            overlay_slot: 
                                                                                Widget:=stack_box:
                                                                                    Orientation:=orientation.Horizontal
                                                                                        Slots:=array:
                                                                                            stack_box_slot: 
                                                                                                Padding:=margin{Left:=InterColumnLeftPadding,Top:=InterColumnTopPadding,Right:=InterColumnRightPadding,Bottom:=InterColumnBottomPadding}
                                                                                                Widget:=overlay:
                                                                                                    Slots:=array:
                                                                                                        overlay_slot: 
                                                                                                            Widget:=texture_block{DefaultDesiredSize:=vector2{X:=DPColumn1SizeX,Y:=DPColumn1SizeY},DefaultImage:=Images.EmptyTexture}
                                                                                                        overlay_slot: 
                                                                                                            Widget:=texture_block{DefaultDesiredSize:=vector2{X:=IconSmallSizeX,Y:=IconSmallSizeY},DefaultImage:=Images.CheeseBread_Icon}
                                                                                            stack_box_slot: 
                                                                                                Padding:=margin{Left:=InterColumnLeftPadding,Top:=InterColumnTopPadding,Right:=InterColumnRightPadding,Bottom:=InterColumnBottomPadding}
                                                                                                Widget:=overlay:
                                                                                                    Slots:=array:
                                                                                                        overlay_slot:
                                                                                                            Widget:=text_block{DefaultText:=BasicMessage("Cheese Bread sales efficiency"), DefaultTextColor:=White, DefaultShadowOffset:=ShadowOffset, DefaultShadowOpacity := DefaultTextShadowOpacity}
                                                                                                        overlay_slot: 
                                                                                                            Widget:=texture_block{DefaultDesiredSize:=vector2{X:=DPColumn2SizeX,Y:=DPColumn2SizeY},DefaultImage:=Images.EmptyTexture}
                                                                                            stack_box_slot: 
                                                                                                Padding:=margin{Left:=InterColumnLeftPadding,Top:=InterColumnTopPadding,Right:=InterColumnRightPadding,Bottom:=InterColumnBottomPadding}
                                                                                                Widget:=overlay:
                                                                                                    Slots:=array:
                                                                                                        overlay_slot:
                                                                                                            Widget:=DP_Row3_Current_TextBlock
                                                                                                        overlay_slot: 
                                                                                                            Widget:=texture_block{DefaultDesiredSize:=vector2{X:=DPColumn4SizeX,Y:=DPColumn4SizeY},DefaultImage:=Images.EmptyTexture}
                                                                                            stack_box_slot: 
                                                                                                Padding:=margin{Left:=InterColumnLeftPadding,Top:=InterColumnTopPadding,Right:=InterColumnRightPadding,Bottom:=InterColumnBottomPadding}
                                                                                                Widget:=overlay:
                                                                                                    Slots:=array:
                                                                                                        overlay_slot:
                                                                                                            Widget:=DP_Row3_Upgrade_TextBlock
                                                                                                        overlay_slot: 
                                                                                                            Widget:=texture_block{DefaultDesiredSize:=vector2{X:=DPColumn5SizeX,Y:=DPColumn5SizeY},DefaultImage:=Images.EmptyTexture}
                                                                                            stack_box_slot: 
                                                                                                Padding:=margin{Left:=InterColumnLeftPadding,Top:=InterColumnTopPadding,Right:=InterColumnRightPadding,Bottom:=InterColumnBottomPadding}
                                                                                                Widget:=overlay:
                                                                                                    Slots:=array:
                                                                                                        overlay_slot: 
                                                                                                            Widget:=texture_block{DefaultDesiredSize:=vector2{X:=DPColumn6SizeX,Y:=DPColumn6SizeY},DefaultImage:=Images.EmptyTexture}
                                                                                                        overlay_slot: 
                                                                                                            HorizontalAlignment:=horizontal_alignment.Left
                                                                                                            VerticalAlignment:=vertical_alignment.Center 
                                                                                                            Padding:=margin{Left:=65.0,Top:=0.0,Right:=0.0,Bottom:=0.0}
                                                                                                            Widget:=DP_Row3_UpgradeCost_TextBlock 
                                                                                                        overlay_slot: 
                                                                                                            HorizontalAlignment:=horizontal_alignment.Left
                                                                                                            VerticalAlignment:=vertical_alignment.Center 
                                                                                                            Padding:=margin{Left:=10.0,Top:=0.0,Right:=0.0,Bottom:=0.0}
                                                                                                            Widget:=texture_block{DefaultDesiredSize:=vector2{X:=IconSmallSizeX,Y:=IconSmallSizeY}, DefaultImage:=Images.AgedSaphire_Icon}
                                                                                                        overlay_slot:
                                                                                                            HorizontalAlignment:=horizontal_alignment.Fill
                                                                                                            VerticalAlignment:=vertical_alignment.Fill   
                                                                                                            Widget:=DP_Row3Button

                                                            stack_box_slot: #Line 7: All titles
                                                                Padding:=margin{Left:=RowLeftPadding,Top:=RowTopPadding,Right:=RowRightPadding,Bottom:=RowBottomPadding}
                                                                    Widget:=overlay:
                                                                        Slots:=array:
                                                                            overlay_slot: 
                                                                                HorizontalAlignment:=horizontal_alignment.Fill
                                                                                VerticalAlignment:=vertical_alignment.Center
                                                                                Widget:=texture_block{DefaultDesiredSize:=vector2{X:=0.0,Y:=LineSizeY},DefaultImage:=Images.Img_LineBackground}
                                                                            overlay_slot: 
                                                                                Widget:=stack_box:
                                                                                    Orientation:=orientation.Horizontal
                                                                                        Slots:=array:
                                                                                            stack_box_slot: 
                                                                                                Padding:=margin{Left:=InterColumnLeftPadding,Top:=InterColumnTopPadding,Right:=InterColumnRightPadding,Bottom:=InterColumnBottomPadding}
                                                                                                Widget:=overlay:
                                                                                                    Slots:=array:
                                                                                                        overlay_slot: 
                                                                                                            Widget:=texture_block{DefaultDesiredSize:=vector2{X:=DPColumn1SizeX,Y:=DPColumn1SizeY},DefaultImage:=Images.EmptyTexture}
                                                                                                        overlay_slot: 
                                                                                                            Widget:=texture_block{DefaultDesiredSize:=vector2{X:=IconSmallSizeX,Y:=IconSmallSizeY},DefaultImage:=Images.PimentoCheeseSandwich_Icon}
                                                                                            stack_box_slot: 
                                                                                                Padding:=margin{Left:=InterColumnLeftPadding,Top:=InterColumnTopPadding,Right:=InterColumnRightPadding,Bottom:=InterColumnBottomPadding}
                                                                                                Widget:=overlay:
                                                                                                    Slots:=array:
                                                                                                        overlay_slot:
                                                                                                            Widget:=text_block{DefaultText:=BasicMessage("Sandwich efficiency"), DefaultTextColor:=White, DefaultShadowOffset:=ShadowOffset, DefaultShadowOpacity := DefaultTextShadowOpacity}
                                                                                                        overlay_slot: 
                                                                                                            Widget:=texture_block{DefaultDesiredSize:=vector2{X:=DPColumn2SizeX,Y:=DPColumn2SizeY},DefaultImage:=Images.EmptyTexture}
                                                                                            stack_box_slot: 
                                                                                                Padding:=margin{Left:=InterColumnLeftPadding,Top:=InterColumnTopPadding,Right:=InterColumnRightPadding,Bottom:=InterColumnBottomPadding}
                                                                                                Widget:=overlay:
                                                                                                    Slots:=array:
                                                                                                        overlay_slot:
                                                                                                            Widget:=DP_Row4_Current_TextBlock
                                                                                                        overlay_slot: 
                                                                                                            Widget:=texture_block{DefaultDesiredSize:=vector2{X:=DPColumn4SizeX,Y:=DPColumn4SizeY},DefaultImage:=Images.EmptyTexture}
                                                                                            stack_box_slot: 
                                                                                                Padding:=margin{Left:=InterColumnLeftPadding,Top:=InterColumnTopPadding,Right:=InterColumnRightPadding,Bottom:=InterColumnBottomPadding}
                                                                                                Widget:=overlay:
                                                                                                    Slots:=array:
                                                                                                        overlay_slot:
                                                                                                            Widget:=DP_Row4_Upgrade_TextBlock
                                                                                                        overlay_slot: 
                                                                                                            Widget:=texture_block{DefaultDesiredSize:=vector2{X:=DPColumn5SizeX,Y:=DPColumn5SizeY},DefaultImage:=Images.EmptyTexture}
                                                                                            stack_box_slot: 
                                                                                                Padding:=margin{Left:=InterColumnLeftPadding,Top:=InterColumnTopPadding,Right:=InterColumnRightPadding,Bottom:=InterColumnBottomPadding}
                                                                                                Widget:=overlay:
                                                                                                    Slots:=array:
                                                                                                        overlay_slot: 
                                                                                                            Widget:=texture_block{DefaultDesiredSize:=vector2{X:=DPColumn6SizeX,Y:=DPColumn6SizeY},DefaultImage:=Images.EmptyTexture}
                                                                                                        overlay_slot: 
                                                                                                            HorizontalAlignment:=horizontal_alignment.Left
                                                                                                            VerticalAlignment:=vertical_alignment.Center 
                                                                                                            Padding:=margin{Left:=TextIcon_LeftMargin}
                                                                                                            Widget:=DP_Row4_UpgradeCost_TextBlock 
                                                                                                        overlay_slot: 
                                                                                                            HorizontalAlignment:=horizontal_alignment.Left
                                                                                                            VerticalAlignment:=vertical_alignment.Center 
                                                                                                            Padding:=margin{Left:=Icon_LeftMargin}
                                                                                                            Widget:=texture_block{DefaultDesiredSize:=vector2{X:=IconSmallSizeX,Y:=IconSmallSizeY}, DefaultImage:=Images.AgedSaphire_Icon}
                                                                                                        overlay_slot:
                                                                                                            HorizontalAlignment:=horizontal_alignment.Fill
                                                                                                            VerticalAlignment:=vertical_alignment.Fill  
                                                                                                            Widget:=DP_Row4Button
                                                            stack_box_slot: #Line 8: All titles
                                                                Padding:=margin{Left:=RowLeftPadding,Top:=RowTopPadding,Right:=RowRightPadding,Bottom:=RowBottomPadding}
                                                                    Widget:=overlay:
                                                                        Slots:=array:
                                                                            overlay_slot: 
                                                                                HorizontalAlignment:=horizontal_alignment.Center
                                                                                VerticalAlignment:=vertical_alignment.Center
                                                                                Widget:=stack_box:
                                                                                    Orientation:=orientation.Horizontal
                                                                                        Slots:=array:
                                                                                            stack_box_slot: 
                                                                                                Padding:=margin{Left:=InterColumnLeftPadding,Top:=InterColumnTopPadding,Right:=InterColumnRightPadding,Bottom:=InterColumnBottomPadding}
                                                                                                Widget:=overlay:
                                                                                                    Slots:=array:
                                                                                                        overlay_slot: 
                                                                                                            Widget:=texture_block{DefaultDesiredSize:=vector2{X:=BigCloseButtonSizeX,Y:=BigCloseButtonSizeY},DefaultImage:=Images.EmptyTexture}
                                                                                                        overlay_slot: 
                                                                                                            HorizontalAlignment:=horizontal_alignment.Fill
                                                                                                            VerticalAlignment:=vertical_alignment.Fill  
                                                                                                            Padding:=margin{Left:=InterColumnLeftPadding,Top:=InterColumnTopPadding,Right:=InterColumnRightPadding,Bottom:=InterColumnBottomPadding}
                                                                                                            Widget:=CloseButton


                                                                                                
                                       


MakeDroneUpgradeUI<constructor>(
    InBakeryMachine : bakery_machines_manager)<transacts> := drone_upgrades_ui :
        BakeryMachinesManager := InBakeryMachine
                                                                                                               