# fill this file with your verse-script
using { /UnrealEngine.com/Temporary/UI }
using { /UnrealEngine.com/Temporary/SpatialMath }
using { /Fortnite.com/UI }
using { /Verse.org/Simulation }
using { /Verse.org/Colors/NamedColors }

# A Verse-authored creative device that can be placed in a level
drop_coal_ui := class():

    PlayersDataManager : player_data_manager = player_data_manager{}
    BakeryMachinesManager : bakery_machines_manager

    TotalCoal_TextBlock : text_block = text_block{DefaultTextColor := CoalTextColor, DefaultShadowOffset:=ShadowOffset, DefaultShadowOpacity := 1.0}

    #Class settings
    var MaybeAgent : ?agent = false
    var MaybeCanvas : ?canvas = false
    var UIIsDisplayed :logic = false

    var ConfirmationToAll : int = 0

    #COAL SLIDER
    DropCoal_ButtonAdd1 : button_regular = button_regular{}
    DropCoal_ButtonAdd5 : button_regular = button_regular{}
    DropCoal_ButtonAdd10 : button_regular = button_regular{}
    DropCoal_ButtonAddAll : button_regular = button_regular{}
    DropCoal_CancelButton : button_loud = button_loud{}

    InitUI() : void =

        Canvas := MakeCoalCanvas()
        set MaybeCanvas = option{Canvas}

        DropCoal_ButtonAdd1.SetText(BasicMessage("Add 1"))
        DropCoal_ButtonAdd5.SetText(BasicMessage("Add 5"))
        DropCoal_ButtonAdd10.SetText(BasicMessage("Add 10"))
        DropCoal_ButtonAddAll.SetText(BasicMessage("Add All"))
        DropCoal_CancelButton.SetText(BasicMessage("CANCEL "))

        Subscribers()

    Subscribers() : void =
    
        DropCoal_ButtonAdd1.OnClick().Subscribe(IntroduceCoal_1) 
        DropCoal_ButtonAdd5.OnClick().Subscribe(IntroduceCoal_5) 
        DropCoal_ButtonAdd10.OnClick().Subscribe(IntroduceCoal_10) 
        DropCoal_ButtonAddAll.OnClick().Subscribe(IntroduceCoal_All) 
        DropCoal_CancelButton.OnClick().Subscribe(Remove_Canvas)

        <#
    CheckMinCoal(Message : widget_message) : void =
        CurrentAmount:=DropCoal_Slider.GetValue()
        if(CurrentAmount<1.0):
            DropCoal_Slider.SetValue(1.0)#>

    IntroduceCoal_1(Message : widget_message) : void =
        TryIntroduceCoal_Amount(1.0)
    IntroduceCoal_5(Message : widget_message) : void =
        TryIntroduceCoal_Amount(5.0)
    IntroduceCoal_10(Message : widget_message) : void =
        TryIntroduceCoal_Amount(10.0)
    IntroduceCoal_All(Message : widget_message) : void =
        if(ConfirmationToAll>=1):
            set ConfirmationToAll = 0
            DropCoal_ButtonAddAll.SetText(BasicMessage("Add All"))
            TryIntroduceCoal_Amount(-1.0)
        else:
            DropCoal_ButtonAddAll.SetText(BasicMessage("Add All Sure?"))
            set ConfirmationToAll = ConfirmationToAll + 1


    TryIntroduceCoal_Amount(Amount : float) : void =
        OriginalAmount_ToAdd := Amount
        var CoalTriedToAdd : float = 0.0
        var CoalOverFlow : logic = false
        var SharedCoalOverflow : logic = false
        if:
            Agent:=MaybeAgent?
            Player := player[Agent]
            Player.IsActive[]
            UKTutorial:=BakeryMachinesManager.MaybeUKTutorial?
            #PlayerData := PlayersDataManager.GetPlayerData[Player]
        then:
            if(BakeryMachinesManager.IsOnUKTutorial?):
                UKTutorial.InsertedCoalEvent(Amount)
                #Remove_CanvasWithoutMessage()  
            else:
                if(Amount=-1.0):
                    set CoalTriedToAdd = BakeryMachinesManager.PlayerDataCache.Inventory_Coal
                else:
                    set CoalTriedToAdd = Amount 
                var UpdatedDataTable : player_data_table = BakeryMachinesManager.GameManager.QuestsAndOrdersDatabase.AddAgentQuest_ProgressV2(Player,BakeryMachinesManager.PlayerDataCache,Amount,USECOAL_TUTORIAL_TYPE)(4)
                
                Log("Tried to add {GetOptimizedMagnitudeLetter(CoalTriedToAdd)}", LOG_LEVEL_DEBUG9)

                if(UpdatedDataTable.Inventory_Coal<=Amount):
                    set CoalTriedToAdd = UpdatedDataTable.Inventory_Coal
                    Log("Corrected ammount because was lower than the inventory ammount New amount: {GetOptimizedMagnitudeLetter(CoalTriedToAdd)}", LOG_LEVEL_DEBUG9)



                var SuperEnergyToAdd_Seconds : float = CoalTriedToAdd*Coal_To_SuperEnergy
                Log("Super Energy ammount tried to add: {GetOptimizedMagnitudeLetter(SuperEnergyToAdd_Seconds)}", LOG_LEVEL_DEBUG9)
                

                #Check coal overflow
                if(BakeryMachinesManager.PlayerDataCache.SuperEnergy_Owned+SuperEnergyToAdd_Seconds>MaxTime_SuperEnergy):
                    set SuperEnergyToAdd_Seconds = MaxTime_SuperEnergy - BakeryMachinesManager.PlayerDataCache.SuperEnergy_Owned
                    Log("Corrected ammount because was > 3600 seconds ({GetOptimizedMagnitudeLetter(BakeryMachinesManager.PlayerDataCache.SuperEnergy_Owned+SuperEnergyToAdd_Seconds)}) New amount: {GetOptimizedMagnitudeLetter(SuperEnergyToAdd_Seconds)}", LOG_LEVEL_DEBUG9)
                    
                    set CoalTriedToAdd = (SuperEnergyToAdd_Seconds*1.0)/(Coal_To_SuperEnergy*1.0)
                    if(IntCoalTriedToAdd := Floor[CoalTriedToAdd]):
                        set CoalTriedToAdd = IntCoalTriedToAdd *1.0

                    Log("So, the new Coal amount to delete from player is {GetOptimizedMagnitudeLetter(CoalTriedToAdd)}", LOG_LEVEL_DEBUG9)
                    set CoalOverFlow = true


                BakeryMachinesManager.DisplayResourcersGather(array{-CoalTriedToAdd},array{COAL},false)
                set UpdatedDataTable=PlayersDataManager.WithdrawResource(UpdatedDataTable,CoalTriedToAdd,COAL)
                set UpdatedDataTable=PlayersDataManager.AddSuperEnergy(UpdatedDataTable,SuperEnergyToAdd_Seconds)
                #PlayersDataManager.SaveNewData(Player,PlayerData,UpdatedDataTable)  
                BakeryMachinesManager.SetNewPlayerDataCache(UpdatedDataTable)
                Log("New Super Energy is {GetOptimizedMagnitudeLetter(BakeryMachinesManager.PlayerDataCache.SuperEnergy_Owned)}", LOG_LEVEL_DEBUG9)
                #PlayersDataManager.SaveAllNewData(Player, PlayerData, UpdatedDataTable, BakeryMachinesManager)
                if(UpdatedDataTable.Inventory_Coal<=0.0):
                    Remove_CanvasWithoutMessage()   
                TotalCoal_TextBlock.SetText(BasicMessage(GetMagnitudeLetter(UpdatedDataTable.Inventory_Coal, "", "")))
                UpdateUI()   
                if(CoalOverFlow?):
                    BakeryMachinesManager.GameManager.DisplayCharacterMessage(Agent,I_CHARACTER,"Added max {GetOptimizedMagnitudeLetter(CoalTriedToAdd)} Coal due to 1h limit. Try again later",?Sound:=MESSAGE_SOUND)
                    Remove_CanvasWithoutMessage()  
                BakeryMachinesManager.RefreshInventoryUI(Agent)
                #AddSharedCoal
                IsSharedOverFlow:=BakeryMachinesManager.GameManager.UpdateSharedCoal()   
                <#
                if(IsSharedOverFlow?, IntTime := Floor[BakeryMachinesManager.GameManager.SuperEnergy_Shared]):
                    TupleTime := TransformToMinutes_Seconds(IntTime)
                    BakeryMachinesManager.GameManager.DisplayCharacterMessage(Agent,I_CHARACTER,"Oops! Generator timer overflow! No worries, {IntTime} Energy left.",?Sound:=MESSAGE_SOUND)#>

                BakeryMachinesManager.SuperEnergyRemaining_Timer.Reset()
                BakeryMachinesManager.SuperEnergyRemaining_Timer.SetActiveDuration(SuperEnergyToAdd_Seconds) 
                BakeryMachinesManager.SuperEnergyRemaining_Timer.SetMaxDuration(SuperEnergyToAdd_Seconds) 
                BakeryMachinesManager.SuperEnergyRemaining_Timer.Reset()
                BakeryMachinesManager.SuperEnergyRemaining_Timer.SetActiveDuration(SuperEnergyToAdd_Seconds) 
                BakeryMachinesManager.SuperEnergyRemaining_Timer.SetMaxDuration(SuperEnergyToAdd_Seconds) 
            
    SetAgent(Agent : agent) : void =
        set ConfirmationToAll = 0
        set MaybeAgent = option{Agent}
        set UIIsDisplayed = false

    UpdateUI() : void =
        if:
            Agent:=MaybeAgent?
            Player := player[Agent]
            Player.IsActive[]
        then:
            TotalCoal:=BakeryMachinesManager.PlayerDataCache.Inventory_Coal
            if(not BakeryMachinesManager.IsOnUKTutorial?):
                TotalCoal_TextBlock.SetText(BasicMessage(GetMagnitudeLetter(TotalCoal, "", "")))
            else:
                TotalCoal_TextBlock.SetText(BasicMessage("∞"))

    Remove_Canvas(Message : widget_message) : void =
        set UIIsDisplayed = false
        if:
            Agent:=MaybeAgent?
            Player:=player[Agent]
            PlayerUI:=GetPlayerUI[Player]
            Canvas := MaybeCanvas?
        then:
            PlayerUI.RemoveWidget(Canvas)
            RemoveDelay_OnRemove()
            IsSharedOverFlow:=BakeryMachinesManager.GameManager.IsSharedCoalOverFlow()   
            if(IsSharedOverFlow?, IntTime := Floor[BakeryMachinesManager.GameManager.SuperEnergy_Shared]):
                TupleTime := TransformToMinutes_Seconds(IntTime)
                BakeryMachinesManager.GameManager.DisplayCharacterMessage(Agent,I_CHARACTER,"Generator timer overflow! No worries, {IntTime} Energy left.",?Sound:=MESSAGE_SOUND)

    Remove_CanvasWithoutMessage() : void =
        set UIIsDisplayed = false
        if:
            Agent:=MaybeAgent?
            Player:=player[Agent]
            PlayerUI:=GetPlayerUI[Player]
            Canvas := MaybeCanvas?
        then:
            PlayerUI.RemoveWidget(Canvas)
            RemoveDelay_OnRemove()
            IsSharedOverFlow:=BakeryMachinesManager.GameManager.IsSharedCoalOverFlow()   
            if(IsSharedOverFlow?, IntTime := Floor[BakeryMachinesManager.GameManager.SuperEnergy_Shared]):
                TupleTime := TransformToMinutes_Seconds(IntTime)
                BakeryMachinesManager.GameManager.DisplayCharacterMessage(Agent,I_CHARACTER,"Generator timer overflow! No worries, {IntTime} Energy left.",?Sound:=MESSAGE_SOUND)

    RemoveDelay_OnAdd() : void =
        DropCoal_CancelButton.SetText(BasicMessage("CLOSE"))


    RemoveDelay_OnRemove() : void =
        DropCoal_CancelButton.SetText(BasicMessage("CLOSE "))
            
    Display_Canvas() : void =
        if(not UIIsDisplayed?):
            set ConfirmationToAll = 0
            set UIIsDisplayed = true
            var Index : int = 0
            if:
                Agent:=MaybeAgent?
                PlayerUI:=GetPlayerUI[player[Agent]]
                Canvas := MaybeCanvas?
            then:
                PlayerUI.AddWidget(Canvas,player_ui_slot{ZOrder:=1, InputMode:=ui_input_mode.All})
                RemoveDelay_OnAdd()   
            UpdateUI()

    MakeCoalCanvas(): canvas=
        Canvas := canvas:
            Slots:=array:
                #Background black image
                canvas_slot:
                    Anchors:=anchors{Minimum:=vector2{X:=0.5,Y:=0.5},Maximum:=vector2{X:=0.5,Y:=0.5}}
                    ZOrder:=0
                    SizeToContent:=true
                    Alignment:=vector2{X:=0.5,Y:=0.5}
                    Widget:=texture_block:
                        DefaultImage:=Images.BlackBackground
                        DefaultDesiredSize:=vector2{X:=BackgroundBlackUISizeX,Y:=BackgroundBlackUISizeY}
                #Background image
                canvas_slot:
                    Anchors:=anchors{Minimum:=vector2{X:=0.5,Y:=0.5},Maximum:=vector2{X:=0.5,Y:=0.5}}
                    ZOrder:=0
                    SizeToContent:=true
                    Offsets:=margin{Left:=0.0,Top:=0.0,Right:=0.0,Bottom:=0.0} #PositionX, PositionY, SizeX, SizeY
                    Alignment:=vector2{X:=0.5,Y:=0.5}
                    Widget:=overlay:
                        Slots:=array:
                            overlay_slot:
                                HorizontalAlignment:=horizontal_alignment.Fill
                                VerticalAlignment:=vertical_alignment.Fill
                                Widget:=texture_block{DefaultImage:=Images.BlackBackground}
                            overlay_slot:
                                HorizontalAlignment:=horizontal_alignment.Fill
                                VerticalAlignment:=vertical_alignment.Fill
                                Padding:=margin{Left:=SelectorBoxPadding,Right:=SelectorBoxPadding,Top:=SelectorBoxPadding,Bottom:=SelectorBoxPadding}
                                Widget:=texture_block{DefaultImage:=Images.Img_MainBackground}
                            overlay_slot: #Screw 0
                                Padding:=margin{Left:=3.5,Top:=3.5,Right:=3.5,Bottom:=3.5}
                                HorizontalAlignment:=horizontal_alignment.Left
                                VerticalAlignment:=vertical_alignment.Top
                                Widget:=texture_block{DefaultImage:=Images.Dialogue_Bolt, DefaultDesiredSize:=vector2{X:=5.0,Y:=5.0}}
                            overlay_slot: #Screw 1
                                Padding:=margin{Left:=3.5,Top:=3.5,Right:=3.5,Bottom:=3.5}
                                HorizontalAlignment:=horizontal_alignment.Right
                                VerticalAlignment:=vertical_alignment.Top
                                Widget:=texture_block{DefaultImage:=Images.Dialogue_Bolt, DefaultDesiredSize:=vector2{X:=5.0,Y:=5.0}}
                            overlay_slot: #Screw 2
                                Padding:=margin{Left:=3.5,Top:=3.5,Right:=3.5,Bottom:=3.5}
                                HorizontalAlignment:=horizontal_alignment.Left
                                VerticalAlignment:=vertical_alignment.Bottom
                                Widget:=texture_block{DefaultImage:=Images.Dialogue_Bolt, DefaultDesiredSize:=vector2{X:=5.0,Y:=5.0}}
                            overlay_slot: #Screw 3
                                Padding:=margin{Left:=3.5,Top:=3.5,Right:=3.5,Bottom:=3.5}
                                HorizontalAlignment:=horizontal_alignment.Right
                                VerticalAlignment:=vertical_alignment.Bottom
                                Widget:=texture_block{DefaultImage:=Images.Dialogue_Bolt, DefaultDesiredSize:=vector2{X:=5.0,Y:=5.0}}
                                <#
                            overlay_slot:
                                HorizontalAlignment:=horizontal_alignment.Fill
                                VerticalAlignment:=vertical_alignment.Fill
                                Widget:=texture_block{DefaultImage:=Images.BlackBackground,DefaultDesiredSize:=vector2{X:=5.0,Y:=5.0}}#>
                            overlay_slot:
                                Padding:=margin{Top:=10.0,Bottom:=10.0,Right:=10.0,Left:=10.0}
                                Widget:=stack_box:
                                    Orientation:=orientation.Vertical
                                        Slots:=array:
                                            stack_box_slot: 
                                                Widget:=overlay:
                                                    Slots:=array:
                                                        overlay_slot:
                                                            Padding:=margin{Top:=5.0,Bottom:=5.0,Right:=5.0,Left:=5.0}
                                                            Widget:=text_block{DefaultText:=BasicMessage("DROP COAL"), DefaultTextColor:=White, DefaultShadowOffset:=ShadowOffset, DefaultShadowOpacity := DefaultTextShadowOpacity}
                                            stack_box_slot: 
                                                Widget:=overlay:
                                                    Slots:=array:
                                                        overlay_slot:
                                                            Padding:=margin{Top:=5.0,Bottom:=5.0,Right:=5.0,Left:=5.0}
                                                            Widget:=text_block{DefaultText:=BasicMessage("Each coal grants {Coal_To_SuperEnergy} seconds of x{GetMagnitudeLetter(SuperEnergyMultiplier_Owned, "", "")} efficiency and "), DefaultTextColor:=CurrentTextColor, DefaultShadowOffset:=ShadowOffset, DefaultShadowOpacity := DefaultTextShadowOpacity}
                                            stack_box_slot: 
                                                Widget:=overlay:
                                                    Slots:=array:
                                                        overlay_slot:
                                                            Padding:=margin{Top:=5.0,Bottom:=5.0,Right:=5.0,Left:=5.0}
                                                            Widget:=text_block{DefaultText:=BasicMessage("Baker speed for you, plus a x{GetMagnitudeLetter(SuperEnergyMultiplier_Shared + 1.0, "", "")} boost for all players."), DefaultTextColor:=CurrentTextColor, DefaultShadowOffset:=ShadowOffset, DefaultShadowOpacity := DefaultTextShadowOpacity}
                                            stack_box_slot:
                                                Padding:=margin{Top:=5.0,Bottom:=5.0,Right:=5.0,Left:=5.0}
                                                Widget:=overlay:
                                                    Slots:=array: 
                                                        overlay_slot:
                                                            Widget:=texture_block{DefaultDesiredSize:=vector2{X:=SizeBoxSaphiresX,Y:=SizeBoxSaphiresY},DefaultImage:=Images.EmptyTexture} 
                                                        overlay_slot:
                                                            HorizontalAlignment:=horizontal_alignment.Fill
                                                            VerticalAlignment:=vertical_alignment.Fill
                                                            Widget:=texture_block{DefaultImage:=Images.Img_LineBackground}  
                                                        overlay_slot:
                                                            HorizontalAlignment:=horizontal_alignment.Left
                                                            VerticalAlignment:=vertical_alignment.Center
                                                            Padding:=margin{Left:=Icon_LeftMargin}
                                                            Widget:=texture_block{DefaultDesiredSize:=vector2{X:=IconSmallSizeX,Y:=IconSmallSizeY},DefaultImage:=Images.Coal_Icon}
                                                        overlay_slot:
                                                            HorizontalAlignment:=horizontal_alignment.Left
                                                            VerticalAlignment:=vertical_alignment.Center
                                                            Padding:=margin{Left:=TextIcon_LeftMargin}
                                                            Widget:=TotalCoal_TextBlock 
                                            stack_box_slot: 
                                                Padding:=margin{Top:=5.0,Bottom:=5.0,Right:=5.0,Left:=5.0}
                                                Widget:=overlay:
                                                    Slots:=array:
                                                        overlay_slot:
                                                            Widget:=texture_block{DefaultImage:=Images.EmptyTexture, DefaultDesiredSize:=vector2{X:=360.0,Y:=70.0}}
                                                        overlay_slot:
                                                            HorizontalAlignment:=horizontal_alignment.Fill
                                                            VerticalAlignment:=vertical_alignment.Fill
                                                            Padding:=margin{Top:=15.0,Bottom:=5.0,Right:=5.0,Left:=5.0}
                                                            Widget:=DropCoal_ButtonAddAll
                                            stack_box_slot: 
                                                Padding:=margin{Top:=5.0,Bottom:=5.0,Right:=5.0,Left:=5.0}
                                                Widget:=overlay:
                                                    Slots:=array:
                                                        overlay_slot:
                                                            Widget:=texture_block{DefaultImage:=Images.EmptyTexture, DefaultDesiredSize:=vector2{X:=360.0,Y:=70.0}}
                                                        overlay_slot:
                                                            HorizontalAlignment:=horizontal_alignment.Fill
                                                            VerticalAlignment:=vertical_alignment.Fill
                                                            Padding:=margin{Top:=15.0,Bottom:=15.0,Right:=5.0,Left:=5.0}
                                                            Widget:=DropCoal_ButtonAdd10
                                            stack_box_slot: 
                                                Padding:=margin{Top:=15.0,Bottom:=15.0,Right:=5.0,Left:=5.0}
                                                Widget:=overlay:
                                                    Slots:=array:
                                                        overlay_slot:
                                                            Widget:=texture_block{DefaultImage:=Images.EmptyTexture, DefaultDesiredSize:=vector2{X:=360.0,Y:=70.0}}
                                                        overlay_slot:
                                                            HorizontalAlignment:=horizontal_alignment.Fill
                                                            VerticalAlignment:=vertical_alignment.Fill
                                                            Padding:=margin{Top:=5.0,Bottom:=5.0,Right:=5.0,Left:=5.0}
                                                            Widget:=DropCoal_ButtonAdd5
                                            stack_box_slot: 
                                                Padding:=margin{Top:=15.0,Bottom:=15.0,Right:=5.0,Left:=5.0}
                                                Widget:=overlay:
                                                    Slots:=array:
                                                        overlay_slot:
                                                            Widget:=texture_block{DefaultImage:=Images.EmptyTexture, DefaultDesiredSize:=vector2{X:=360.0,Y:=70.0}}
                                                        overlay_slot:
                                                            HorizontalAlignment:=horizontal_alignment.Fill
                                                            VerticalAlignment:=vertical_alignment.Fill
                                                            Padding:=margin{Top:=5.0,Bottom:=5.0,Right:=5.0,Left:=5.0}
                                                            Widget:=DropCoal_ButtonAdd1
                                            stack_box_slot: 
                                                Padding:=margin{Top:=15.0,Bottom:=15.0,Right:=5.0,Left:=5.0}
                                                Widget:=overlay:
                                                    Slots:=array:
                                                        overlay_slot:
                                                            Widget:=texture_block{DefaultImage:=Images.EmptyTexture, DefaultDesiredSize:=vector2{X:=360.0,Y:=80.0}}
                                                        overlay_slot:
                                                            HorizontalAlignment:=horizontal_alignment.Fill
                                                            VerticalAlignment:=vertical_alignment.Fill
                                                            Padding:=margin{Top:=5.0,Bottom:=5.0,Right:=5.0,Left:=5.0}
                                                            Widget:=DropCoal_CancelButton


MakeDropCoalUI<constructor>(InBakeryMachinesManager : bakery_machines_manager)<transacts> := drop_coal_ui :
    BakeryMachinesManager := InBakeryMachinesManager