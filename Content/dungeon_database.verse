using { /UnrealEngine.com/Temporary/UI }
using { /UnrealEngine.com/Temporary/SpatialMath }
using { /Verse.org/Random }
using { /Verse.org/Simulation }
using { /Verse.org/Assets }

#All challenges DataBase

TotalDungeonTime : int = 60

dungeon_database := class():

    PlayerDataManager : player_data_manager = player_data_manager{}
    DungeonMobDistanceDetection : float = 512.0
    var MaybeGameManager : ?game_manager = false

    MaxTimeToEnter : int = 30
    var AllDungeonsStats : []tuple(string,float,float,int,[]int,texture,int,int,int)= 
        array{
            #DungeonName,MultiplierChest1,MultiplierChest2,MaxTime,Resourcers,Image,NumMaxSecondaryMobs,LVBoss,LVMobs
            Dungeon0Stat : tuple(string,float,float,int,[]int,texture,int,int,int) = ("City Sewers",20.0,50.0,180,array{MUTANTMATERIAL,WATER,FLOUR,CHEESE},Images.Dungeon_00,70,2,2)
            Dungeon1Stat : tuple(string,float,float,int,[]int,texture,int,int,int) = ("Hidden Works",25.0,100.0,240,array{MUTANTMATERIAL,WATER,FLOUR},Images.Dungeon_1,70,3,3)
            Dungeon2Stat : tuple(string,float,float,int,[]int,texture,int,int,int) = ("The Basement",30.0,110.0,260,array{MUTANTMATERIAL,WATER,FLOUR,CHEESE},Images.Dungeon_2,70,5,5)
            Dungeon3Stat : tuple(string,float,float,int,[]int,texture,int,int,int) = ("Dig Treasures",35.0,115.0,300,array{MUTANTMATERIAL,WATER,FLOUR,CHEESE},Images.Dungeon_3,70,10,10)
            Dungeon4Stat : tuple(string,float,float,int,[]int,texture,int,int,int) = ("Desert Cave",40.0,120.0,300,array{MUTANTMATERIAL,WATER,FLOUR,CHEESE,PIMENTOS},Images.Dungeon_4,70,10,10)
            Dungeon5Stat : tuple(string,float,float,int,[]int,texture,int,int,int) = ("Experiments",45.0,125.0,300,array{MUTANTMATERIAL,WATER,FLOUR,CHEESE,PIMENTOS},Images.Dungeon_5,70,10,10)
            Dungeon6Stat : tuple(string,float,float,int,[]int,texture,int,int,int) = ("Hidden Batch",50.0,150.0,300,array{MUTANTMATERIAL,WATER,FLOUR,CHEESE,PIMENTOS},Images.Dungeon_6,70,10,10)
        }

    SetGameManager(GameManager : game_manager) : void =
        set MaybeGameManager = option{GameManager}

    GetDungeonMobLevel(NumDungeon : int) : int =
        return NumDungeon * 5

    CalculeDungeon_RemainingTimeToJoin(NumDungeon : int, TotalTime : int) : int =
        var TimeRemaining : int = 99
        if(DungeonStats:=AllDungeonsStats[NumDungeon],MaxTimeToPlay:=DungeonStats(3)):
            set TimeRemaining=TotalTime-(MaxTimeToPlay-MaxTimeToEnter)
        return TimeRemaining

    GetEnemyWave(NumMobs : int) : []int =
        #var AllArrayTypes : []int = array{PLANT_MOB}
        
        #set AllArrayTypes = AllArrayTypes + array{PLANT_MOB} + array{GROUND_MOB} #Two ground mobs

        
        var AllArrayTypes : []int = array{}
        var FinalMobsToSpawn: int = NumMobs
        if(FinalMobsToSpawn>5):
            set FinalMobsToSpawn = 5

        if(FinalMobsToSpawn=1):
            set AllArrayTypes = AllArrayTypes + array{PLANT_MOB} #Only one plant
        else if(FinalMobsToSpawn=2):
            set AllArrayTypes = AllArrayTypes + array{GROUND_MOB} + array{GROUND_MOB} #Two ground mobs
        else:
            set AllArrayTypes = AllArrayTypes + array{DRUID_MOB} #One druid, the others ground mobs
            for(Index:=1..FinalMobsToSpawn-1):
                set AllArrayTypes = AllArrayTypes + array{GROUND_MOB}
        return AllArrayTypes

    GetMobsLV(NumDungeon : int) : tuple(int,int) =
        var LVBoss : int = 1
        var LVMobs : int = 1
        if(DungeonStats:=AllDungeonsStats[NumDungeon]):
            set LVBoss = DungeonStats(7)
            set LVMobs = DungeonStats(8)
        return (LVBoss,LVMobs)

    GetNumMobsToBeSpawn(NumDungeon : int, NumPositions : int) : []int =
        var AllMobsToSpawnPositions : []int  = array{} 
        #var NumTotalMobs : int = 0
        
        if:
            TupleStats:=AllDungeonsStats[NumDungeon]
            #set NumTotalMobs=TupleStats(6)
            RandomNumPositionsWhereToSpawn := Ceil[NumPositions*0.85]
        then:
            ##Print("RandomNumPositionsWhereToSpawn: {RandomNumPositionsWhereToSpawn} or {Positions} total positions")
            #PosibleNumMobsPerPosition:=5#(NumTotalMobs*1.0)/(RandomNumPositionsWhereToSpawn*1.0)
            ##Print("PosibleNumMobsPerPosition: {PosibleNumMobsPerPosition}")
            for(Index:=0..NumPositions-4): #Minus 3 because the three first positions we do not spawn enemies
                set AllMobsToSpawnPositions = AllMobsToSpawnPositions + array{0} 
            for(Index:=0..RandomNumPositionsWhereToSpawn-1):
                var RandomNumOfMobs : int = GetRandomInt(2,6)
                if(Index*1.0<(RandomNumPositionsWhereToSpawn-1)*0.2):
                    set RandomNumOfMobs = 1
                <#
                if:
                    #RandomNumOfMobs := Floor[PosibleNumMobsPerPosition*GetRandomFloat(0.75,1.5)]#>
                if(set AllMobsToSpawnPositions[Index] = RandomNumOfMobs):
            ##Print("AllMobsToSpawnPositions.Length is {AllMobsToSpawnPositions.Length}")
            set AllMobsToSpawnPositions = Shuffle(AllMobsToSpawnPositions)
            set AllMobsToSpawnPositions = array{0} + array{0} + array{0} + AllMobsToSpawnPositions #Add to positions
            #for(Index:=0..AllMobsToSpawnPositions.Length-1, RandomNumOfMobs := AllMobsToSpawnPositions[Index]):
                ##Print("RandomNumOfMobs: {RandomNumOfMobs} on position {Index}")
            #for(Index1:=0..AllMobsToSpawnPositions.Length-1, NumMobs:=AllMobsToSpawnPositions[Index1]):
            #    #Print("On Position {Index1} will be spawned {NumMobs} mobs")
        else:
            PrintFatalError("GetNumMobsToBeSpawn to be spawn failed")
        return AllMobsToSpawnPositions

                
    GetDungeonName(NumDungeon : int) : string = 
        var Name : string  = ""
        if:
            TupleStats:=AllDungeonsStats[NumDungeon]
        then:
            return TupleStats(0)
        return Name

    GetDungeonImage(NumDungeon : int) : texture = 
        var DungeonImage : texture = Images.EmptyTexture
        if:
            TupleStats:=AllDungeonsStats[NumDungeon]
        then:
            return TupleStats(5)
        return DungeonImage

    GetDungeonMaxTime(NumDungeon : int) : int = 
        var TotalTime : int = 99
        if(TupleDungeon:=AllDungeonsStats[NumDungeon]):
            set TotalTime = TupleDungeon(3)
        return TotalTime

    GetSecondaryChestMultiplier(NumDungeon : int) : float = 
        var Multiplier : float = 1.0
        if(TupleDungeon:=AllDungeonsStats[NumDungeon]):
            set Multiplier = TupleDungeon(2)
        return Multiplier

    GetMainChestMultiplier(NumDungeon : int) : float = 
        var Multiplier : float = 1.0
        if(TupleDungeon:=AllDungeonsStats[NumDungeon]):
            set Multiplier = TupleDungeon(1)
        return Multiplier

    GetAndSaveChestAwards(Player : player, ChestType : int, NumDungeon : int, HBLevel : int) : tuple([]float,[]int) =
        #AVGHits := (MaxNum_AVGHit+MinNum_AVGHit)/2.0
        #ChestHitsReward := AVGHits/3.0

        var AllAmounts : []float = array{}
        var AllTypes : []int = array{}

        if(GameManager:=MaybeGameManager?, Agent := player[Player], BakeryMachinesManager:=GameManager.GetAgentBakery_Transacted(Agent)?, Evolution:=BakeryMachinesManager.PlayerDataCache.AllEvolvedDungeons[NumDungeon]):
        
            ##Print("Chest button detected")
            if(TupleDungeon:=AllDungeonsStats[NumDungeon]):
                #Decide amounts
                var MultiplierHits : float = 1.0
                if(ChestType=TREASURE_TYPE_BUTTON):
                    set MultiplierHits = TupleDungeon(1)
                else:
                    set MultiplierHits = TupleDungeon(2)

                AmountMutantMaterial:=CheckFinalMutantMaterialAward(HBLevel,MultiplierHits*0.25,1.0)
                TotalMutantMaterial := AmountMutantMaterial + AmountMutantMaterial*DungeonEvolution_Base*Evolution

                ResourcersToGive:=BakeryMachinesManager.GetBakeryPickaxeDamage(HBLevel,false,true)(0)*MultiplierHits
                TotalResourcersToGive:= ResourcersToGive + ResourcersToGive*DungeonEvolution_Base*Evolution

                set AllAmounts = array{TotalResourcersToGive,TotalResourcersToGive,TotalResourcersToGive,TotalResourcersToGive,TotalMutantMaterial,5.0}
                set AllTypes = array{WATER,FLOUR,CHEESE,PIMENTOS,MUTANTMATERIAL,COAL}

            #Save Resourcers #TODO LIMIT!!!!!!!!!!!!!!!!!!!!!!!
            if(Player.IsActive[],ValidGameManager:=MaybeGameManager?):
                MaybeBakery:=ValidGameManager.GetAgentBakery(Agent)
                if(ValidBakery:=MaybeBakery?):
                    var PlayerDataUpdated : player_data_table = ValidBakery.PlayerDataCache
                    #for(Index:=0..ResourceTypes.Length-1, Type:=ResourceTypes[Index], Amount := ResourceAmount[Index]):

                    ValidBakery.SaveResourcers(Agent,AllTypes,AllAmounts,?AutoBackpack:=true)
                    #    set PlayerDataUpdated = PlayerDataManager.AddResourcers(PlayerDataUpdated, Amount, Type, 999999.9)(0)
                    #PlayerDataManager.SaveNewData(Player,PlayerData,PlayerDataUpdated)
                    #ValidBakery.SetNewPlayerDataCache(PlayerDataUpdated)
                    #PlayerDataManager.SaveAllNewData(Player, PlayerData, PlayerDataUpdated, Bakery)
        return (AllAmounts,AllTypes)
