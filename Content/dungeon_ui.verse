using { /UnrealEngine.com/Temporary/UI }
using { /Verse.org/Simulation }
using { /Fortnite.com/UI }
using { /UnrealEngine.com/Temporary/SpatialMath }
using { /Verse.org/Colors/NamedColors }

# fill this file with your verse-script
dungeon_ui := class():

    var BakeryManager : bakery_machines_manager

    var IsUIDisplayed : logic = false

    var Opened_NumDungeon : int = -1

    var MaybeAgent : ?agent = false
    var MaybeDungeonCanvas : ?canvas = false

    TotalCoal_TextBlock : text_block = text_block{DefaultTextColor:=CoalTextColor, DefaultShadowOffset:=ShadowOffset, DefaultShadowOpacity := 1.0}
    DungeonImage : texture_block = texture_block{DefaultImage:=Images.EmptyTexture, DefaultDesiredSize:=vector2{X:=300.0,Y:=200.0}}
    Availability_TextBlock : text_block = text_block{DefaultTextColor:=Green, DefaultShadowOffset:=ShadowOffset, DefaultShadowOpacity := 1.0}
    Time_TextBlock : text_block = text_block{DefaultTextColor:=White, DefaultShadowOffset:=ShadowOffset, DefaultShadowOpacity := 1.0}
    NumPlayers_TextBlock : text_block = text_block{DefaultTextColor:=White, DefaultShadowOffset:=ShadowOffset, DefaultShadowOpacity := 1.0}
    DungeonName_TextBlock : text_block = text_block{DefaultTextColor:=White, DefaultShadowOffset:=ShadowOffset, DefaultShadowOpacity := 1.0}
    DungeonGatherInfo_Rewards : text_block = text_block{DefaultTextColor:=White, DefaultShadowOffset:=ShadowOffset, DefaultShadowOpacity := 1.0}
    DungeonBigChest_Rewards : text_block = text_block{DefaultTextColor:=UpgradeTextColor, DefaultShadowOffset:=ShadowOffset, DefaultShadowOpacity := 1.0}
    DungeonSmallChest_Rewards : text_block = text_block{DefaultTextColor:=UpgradeTextColor, DefaultShadowOffset:=ShadowOffset, DefaultShadowOpacity := 1.0}
    DungeonOriginalLV_TextBlock : text_block = text_block{DefaultTextColor:=White, DefaultShadowOffset:=ShadowOffset, DefaultShadowOpacity := 1.0}
    DungeonLV_TextBlock : text_block = text_block{DefaultTextColor:=White, DefaultShadowOffset:=ShadowOffset, DefaultShadowOpacity := 1.0}
    DungeonEvolve_TotalAgedSapphires_TextBlock : text_block = text_block{DefaultTextColor:=CoalTextColor, DefaultShadowOffset:=ShadowOffset, DefaultShadowOpacity := DefaultTextShadowOpacity}

    var ConfirmationToEvolve : int = 0
    ButtonEvolve : button_quiet = button_quiet{}
    ButtonPlay : button_regular = button_regular{}
    ButtonCancel : button_loud = button_loud{}
    ButtonExit : button_loud = button_loud{}

    InitUI() : void =
        ButtonCancel.SetText(BasicMessage("CANCEL JOIN"))
        ButtonExit.SetText(BasicMessage("CLOSE "))
        Canvas := MakeDungeonUICanvas()
        set MaybeDungeonCanvas = option{Canvas}
        InitSubscribes()
        UpdateUI()

    OnPlayerRemoved() : void =
        set MaybeAgent = false
        Remove_CanvasWithoutMessage()

    TryEvolveDungeon(Message : widget_message) : void =
        if:
            Player := Message.Player
            Agent := agent[Player]
            #PlayerData := BakeryManager.GameManager.PlayersDataManager.GetPlayerData[Player]
            EvolvedNum:=BakeryManager.PlayerDataCache.AllEvolvedDungeons[Opened_NumDungeon]
            Dungeon_HardBreadLevel:=BakeryManager.GameManager.GeneralMobsManager.AllDungeons_HBAreaLevel[Opened_NumDungeon]
            CurrentDungeonLevel:=Opened_NumDungeon+EvolvedNum
        then:
            if(EvolvedNum<20):  
                Cost := MainFormula_EvolveDungeonCost(Dungeon_HardBreadLevel,EvolvedNum+1)
                if(Cost>BakeryManager.PlayerDataCache.Inventory_Coal):
                    DungeonEvolve_TotalAgedSapphires_TextBlock.SetTextColor(NotEnoughTextColor)
                else:
                    if(ConfirmationToEvolve>0):
                        set ConfirmationToEvolve = 0
                        var NewData : player_data_table =BakeryManager.GameManager.PlayersDataManager.EvolveDungeon(BakeryManager.PlayerDataCache,Opened_NumDungeon)
                        set NewData = BakeryManager.GameManager.PlayersDataManager.WithdrawResource(NewData,Cost,COAL)
                        BakeryManager.SetNewPlayerDataCache(NewData)
                        BakeryManager.GameManager.VFXManager.SmallObjetive_AudioPlayer.Play(Agent)
                        UpdateUI()
                    else:
                        DungeonEvolve_TotalAgedSapphires_TextBlock.SetText(BasicMessage(GetMagnitudeLetter(Cost," x"," Are you sure?")))
                        set ConfirmationToEvolve = ConfirmationToEvolve + 1
                    DungeonEvolve_TotalAgedSapphires_TextBlock.SetTextColor(CoalTextColor)
            else:
                DungeonEvolve_TotalAgedSapphires_TextBlock.SetText(BasicMessage("MAXED"))
                DungeonEvolve_TotalAgedSapphires_TextBlock.SetTextColor(MaxedColor)
                BakeryManager.GameManager.VFXManager.ErrorSound_AudioPlayer.Play(Agent)
        BakeryManager.RefreshAllUI()

    CalculeDungeonEvolveCost() : void = 
        if(Opened_NumDungeon<>-1):
            if:
                OwnerAgent:=BakeryManager.MaybeAgentOwner?
                Player := player[OwnerAgent]
                EvolvedNum:=BakeryManager.PlayerDataCache.AllEvolvedDungeons[Opened_NumDungeon]
                CurrentDungeonLevel:=Opened_NumDungeon+EvolvedNum
                Dungeon_HardBreadLevel:=BakeryManager.GameManager.GeneralMobsManager.AllDungeons_HBAreaLevel[Opened_NumDungeon]
            then:
                if(EvolvedNum<20):
                    Cost := MainFormula_EvolveDungeonCost(Dungeon_HardBreadLevel,EvolvedNum+1)
                    DungeonEvolve_TotalAgedSapphires_TextBlock.SetText(BasicMessage(GetMagnitudeLetter(Cost," x"," Evolve Dungeon + 5% Rewards")))
                    if(Cost>BakeryManager.PlayerDataCache.Inventory_Coal):
                        DungeonEvolve_TotalAgedSapphires_TextBlock.SetTextColor(NotEnoughTextColor)
                    else:
                        DungeonEvolve_TotalAgedSapphires_TextBlock.SetTextColor(CoalTextColor)
                else:
                    DungeonEvolve_TotalAgedSapphires_TextBlock.SetText(BasicMessage("MAXED"))
                    DungeonEvolve_TotalAgedSapphires_TextBlock.SetTextColor(MaxedColor)

    UpdateUI() : void =
        CalculeDungeonEvolveCost()
        ButtonPlay.SetText(BasicMessage("CLEAN"))
        SetDungeonLV()
        DungeonImage.SetImage(BakeryManager.GameManager.GeneralMobsManager.DungeonDataBase.GetDungeonImage(Opened_NumDungeon))
        MaxTimeToPlay:=BakeryManager.GameManager.GeneralMobsManager.DungeonDataBase.GetDungeonMaxTime(Opened_NumDungeon)
        TotalCoal_TextBlock.SetText(BasicMessage(GetOptimizedMagnitudeLetter(BakeryManager.PlayerDataCache.Inventory_Coal)))
        if:
            Agent := MaybeAgent?
            Player:=player[Agent]
            Player.IsActive[]
            IsPublic:=BakeryManager.GameManager.GeneralMobsManager.AllDungeons_IsPublic[Opened_NumDungeon]
            HasStarted:=BakeryManager.GameManager.GeneralMobsManager.AllDungeons_HasStarted[Opened_NumDungeon]
            NumPlayersPlaying:=BakeryManager.GameManager.GeneralMobsManager.AllDungeons_AgentsPlaying[Opened_NumDungeon].Length
            NumPlayersWaiting:=BakeryManager.GameManager.GeneralMobsManager.AllDungeons_AgentsWaitingToJoin[Opened_NumDungeon].Length
            TotalTime:=BakeryManager.GameManager.GeneralMobsManager.AllDungeons_TotalTime[Opened_NumDungeon]
            CountToStart:=BakeryManager.GameManager.GeneralMobsManager.AllDungeons_CountToStartTime[Opened_NumDungeon]                
            RemaingTimeToJoin:=TotalTime-(MaxTimeToPlay-(BakeryManager.GameManager.GeneralMobsManager.DungeonDataBase.MaxTimeToEnter))
        then:
            SetTime(TotalTime)
            if(HasStarted?):
                if(IsPublic?):
                    ##Print("This is a public game")
                    ButtonPlay.SetText(BasicMessage("JOIN!"))
                    RemainingTimeToJoin:=BakeryManager.GameManager.GeneralMobsManager.DungeonDataBase.CalculeDungeon_RemainingTimeToJoin(Opened_NumDungeon,TotalTime)
                    ##Print("RemaingTimeToJoin: {RemainingTimeToJoin}")
                    if(RemainingTimeToJoin>0):
                        Availability_TextBlock.SetText(BasicMessage("AVAILABLE TO JOIN ({RemainingTimeToJoin})"))
                        Availability_TextBlock.SetTextColor(Green)
                    else:
                        ButtonPlay.SetText(BasicMessage("YOU CAN NOT JOIN"))
                        Availability_TextBlock.SetText(BasicMessage("TIME TO JOIN OVER"))
                        Availability_TextBlock.SetTextColor(Red)
                else: 
                    ##Print("This is not a public game")
                    ButtonPlay.SetText(BasicMessage("YOU CAN NOT JOIN"))
                    Availability_TextBlock.SetText(BasicMessage("THE HOST SET THE GAME AS PRIVATE"))
                    Availability_TextBlock.SetTextColor(Red)
                NumPlayers_TextBlock.SetText(BasicMessage("{NumPlayersPlaying} Players in the dungeon"))
            else:
                ##Print("The dungeon is ready to start")
                if(NumPlayersWaiting>0):
                    NumPlayers_TextBlock.SetText(BasicMessage("{NumPlayersWaiting} Players waiting to play"))
                    Time_TextBlock.SetText(BasicMessage("Starting in: {CountToStart}"))
                    if:
                        MaybeStartingAgent := BakeryManager.GameManager.GeneralMobsManager.AllDungeons_MaybeStartingAgent[Opened_NumDungeon]
                        StartingAgent := MaybeStartingAgent?
                        StartingAgent = Agent
                    then:
                        Availability_TextBlock.SetText(BasicMessage("HOSTING GAME"))
                        Availability_TextBlock.SetTextColor(Yellow)
                        if(not IsPublic?):
                            ButtonPlay.SetText(BasicMessage("Ask for help"))
                        else:
                            BakeryManager.GameManager.AskForHelp(Agent)
                            ButtonPlay.SetText(BasicMessage("Done!"))
                    else:
                        var AlreadyJoined : logic = false
                        for(OtherAgentWaiting : BakeryManager.GameManager.GeneralMobsManager.AllDungeons_AgentsWaitingToJoin[Opened_NumDungeon], OtherAgentWaiting=Agent):
                            set AlreadyJoined = true
                        if(AlreadyJoined?):
                            SetDungeonLV()
                            #Print("You joined")
                            ButtonPlay.SetText(BasicMessage("JOINED"))
                            Availability_TextBlock.SetText(BasicMessage("JOINED"))
                            Availability_TextBlock.SetTextColor(Green)
                        else:
                            ##Print("You did not join yet")
                            if(IsPublic?):
                                ##Print("This is a public game")
                                ButtonPlay.SetText(BasicMessage("JOIN!"))  
                                Availability_TextBlock.SetText(BasicMessage("AVAILABLE TO JOIN"))
                                Availability_TextBlock.SetTextColor(Green)
                            else: 
                                ##Print("This is not a public game")
                                
                                ButtonPlay.SetText(BasicMessage("YOU CAN NOT JOIN"))
                                Availability_TextBlock.SetText(BasicMessage("THE HOST SET THE GAME AS PRIVATE"))
                                Availability_TextBlock.SetTextColor(Red)
                else:
                    Time_TextBlock.SetText(BasicMessage("--:--"))
                    NumPlayers_TextBlock.SetText(BasicMessage("{NumPlayersPlaying} Players waiting to play"))
                    if:
                        Player.IsActive[]
                        BakeryManager.PlayerDataCache.Dungeon_Accumulated > 0
                    then:
                        Availability_TextBlock.SetText(BasicMessage("AVAILABLE TO HOST | Tokens: {BakeryManager.PlayerDataCache.Dungeon_Accumulated}"))
                        Availability_TextBlock.SetTextColor(Green)
                    else:
                        Availability_TextBlock.SetText(BasicMessage("NO CLEANING PERMIT | Tokens: {BakeryManager.PlayerDataCache.Dungeon_Accumulated}"))
                        Availability_TextBlock.SetTextColor(Red)
        else:
            PrintWarning("OnDungeon UI: Can not update the UI of player")

    SetDungeonLV() : void =
        StringForName := BakeryManager.GameManager.GeneralMobsManager.DungeonDataBase.GetDungeonName(Opened_NumDungeon)
        MainChestMultiplier := BakeryManager.GameManager.GeneralMobsManager.DungeonDataBase.GetMainChestMultiplier(Opened_NumDungeon)
        SecondaryChestMultiplier := BakeryManager.GameManager.GeneralMobsManager.DungeonDataBase.GetSecondaryChestMultiplier(Opened_NumDungeon)
        if:
            MaybeStartingAgent := BakeryManager.GameManager.GeneralMobsManager.AllDungeons_MaybeStartingAgent[Opened_NumDungeon]
            StartingAgent := MaybeStartingAgent?
        then:
            if:
                
                ValidStartingPlayer := player[StartingAgent]
                EvolvedNum:=BakeryManager.PlayerDataCache.AllEvolvedDungeons[Opened_NumDungeon]
                AllButtons:=BakeryManager.GameManager.GeneralMobsManager.AllDungeons_Buttons[Opened_NumDungeon]
                HBLevelArea := BakeryManager.GameManager.GeneralMobsManager.AllDungeons_HBAreaLevel[Opened_NumDungeon]
            then:
                
                #DungeonOriginalLV_TextBlock.SetText(BasicMessage("LV: {Opened_NumDungeon+1}"))
                DungeonLV_TextBlock.SetText(BasicMessage(" +{ToTruncatedString(EvolvedNum*DungeonEvolution_Base*100.0,0)}% Rewards"))

                MainChest_TotalNumHits := MainChestMultiplier + MainChestMultiplier*DungeonEvolution_Base*EvolvedNum
                SecondaryChest_TotalNumHits := SecondaryChestMultiplier + SecondaryChestMultiplier*DungeonEvolution_Base*EvolvedNum

                DungeonGatherInfo_Rewards.SetText(BasicMessage("Based on LV {HBLevelArea} Slate Bread Gather"))
                DungeonBigChest_Rewards.SetText(BasicMessage("x1 Big Chest | x{GetOptimizedMagnitudeLetter(MainChest_TotalNumHits)} Gather"))
                DungeonSmallChest_Rewards.SetText(BasicMessage("x{AllButtons.Length-3} Small Chests | x{GetOptimizedMagnitudeLetter(SecondaryChest_TotalNumHits)} Gather"))

                
                if(ExpectedDamageReceived_PerHit:=BakeryManager.GameManager.GeneralMobsManager.AllDungeons_ExpectedHammerDamage[Opened_NumDungeon]):
                    MiniBossCurrentHealth:=CalculeMobHealth(MINIBOSS_MOB,ExpectedDamageReceived_PerHit,HBLevelArea)  
                    DungeonName_TextBlock.SetText(BasicMessage("{StringForName} | Boss Health: {GetOptimizedMagnitudeLetter(MiniBossCurrentHealth)}"))
                else:
                    PrintFatalError("ExpectedDamageReceived_PerHit Invalid! HBLevelArea Invalid {HBLevelArea}")
        else:
            if:
                OwnerAgent:=BakeryManager.MaybeAgentOwner?
                ValidPlayer := player[OwnerAgent]
                EvolvedNum:=BakeryManager.PlayerDataCache.AllEvolvedDungeons[Opened_NumDungeon]
                AllButtons:=BakeryManager.GameManager.GeneralMobsManager.AllDungeons_Buttons[Opened_NumDungeon]
                HBLevelArea := BakeryManager.GameManager.GeneralMobsManager.AllDungeons_HBAreaLevel[Opened_NumDungeon]
            then:
                DungeonName_TextBlock.SetText(BasicMessage(StringForName))
                #DungeonOriginalLV_TextBlock.SetText(BasicMessage("LV: {Opened_NumDungeon+1}"))
                DungeonLV_TextBlock.SetText(BasicMessage(" +{ToTruncatedString(EvolvedNum*DungeonEvolution_Base*100.0,0)}% Rewards"))

                MainChest_TotalNumHits := MainChestMultiplier + MainChestMultiplier*DungeonEvolution_Base*EvolvedNum
                SecondaryChest_TotalNumHits := SecondaryChestMultiplier + SecondaryChestMultiplier*DungeonEvolution_Base*EvolvedNum

                DungeonGatherInfo_Rewards.SetText(BasicMessage("Based on LV {HBLevelArea} Slate Bread Gather"))
                DungeonBigChest_Rewards.SetText(BasicMessage("x1 Big Chest | x{GetOptimizedMagnitudeLetter(MainChest_TotalNumHits)} Gather"))
                DungeonSmallChest_Rewards.SetText(BasicMessage("x{AllButtons.Length-3} Small Chests | x{GetOptimizedMagnitudeLetter(SecondaryChest_TotalNumHits)} Gather"))
                if(Color:=AllTierColors[EvolvedNum]):
                    DungeonLV_TextBlock.SetTextColor(Color)

                if(ExpectedDamageReceived_PerHit:=BakeryManager.GameManager.GeneralMobsManager.AllDungeons_ExpectedHammerDamage[Opened_NumDungeon]):
                    MiniBossCurrentHealth:=CalculeMobHealth(MINIBOSS_MOB,ExpectedDamageReceived_PerHit,HBLevelArea)  
                    DungeonName_TextBlock.SetText(BasicMessage("{StringForName} | Boss Health: {GetOptimizedMagnitudeLetter(MiniBossCurrentHealth)}"))
                else:
                    PrintFatalError("ExpectedDamageReceived_PerHit Invalid! HBLevelArea Invalid {HBLevelArea}")

    RunNoCleaningPermit()<suspends> : void =
        ButtonPlay.SetText(BasicMessage("No cleaning permit!"))
        Sleep(1.0)
        ButtonPlay.SetText(BasicMessage("CLEAN"))
        UpdateUI()

    SetTime(Time : int) : void =
        TupleTime:=TransformToMinutes_Seconds(Time)
        Time_TextBlock.SetText(BasicMessage("{TupleTime(0)}:{TupleTime(1)}"))

    InitSubscribes() : void =
        ButtonEvolve.OnClick().Subscribe(TryEvolveDungeon)
        ButtonExit.OnClick().Subscribe(Remove_Canvas)
        ButtonPlay.OnClick().Subscribe(PlayPressed)
        ButtonCancel.OnClick().Subscribe(CancelJoin)

    PlayPressed(Message : widget_message) : void =
        if(Agent:=agent[Message.Player]):
            BakeryManager.GameManager.GeneralMobsManager.DungeonUI_PlayButtonPressed(Agent,Opened_NumDungeon)
            #UpdateUI()

    CancelJoin(Message : widget_message) : void =
        if(Agent:=agent[Message.Player]):
            BakeryManager.GameManager.GeneralMobsManager.DungeonUI_CancelButtonPressed(Agent,Opened_NumDungeon)
            #Remove_Canvas(Message)


    SetAgent(Agent : agent) : void=
        Remove_CanvasWithoutMessage()
        set MaybeAgent = option{Agent}

    Remove_Canvas(Message : widget_message) : void =
        set IsUIDisplayed = false
        if:
            Agent := MaybeAgent?
            Player:=player[Agent]
            PlayerUI:=GetPlayerUI[Player]
            Canvas := MaybeDungeonCanvas?
        then:
            PlayerUI.RemoveWidget(Canvas)
            RemoveDelay_OnRemove()


    Remove_CanvasWithoutMessage() : void =
        set IsUIDisplayed = false
        if:
            Agent := MaybeAgent?
            Player:=player[Agent]
            PlayerUI:=GetPlayerUI[Player]
            Canvas := MaybeDungeonCanvas?
        then:
            PlayerUI.RemoveWidget(Canvas)
            RemoveDelay_OnRemove()

    RemoveDelay_OnAdd() : void =
        ButtonExit.SetText(BasicMessage("CLOSE"))

    RemoveDelay_OnRemove() : void =
        ButtonExit.SetText(BasicMessage("CLOSE "))

    Display_Canvas(NumDungeon : int) : void =
        if(not IsUIDisplayed?):
            set Opened_NumDungeon = NumDungeon
            var Index : int = 0 
            if:
                Agent := MaybeAgent?
                PlayerUI:=GetPlayerUI[player[Agent]]
                Canvas := MaybeDungeonCanvas?
            then:
                PlayerUI.AddWidget(Canvas,player_ui_slot{ZOrder:=0, InputMode := ui_input_mode.All})
                RemoveDelay_OnAdd()
        UpdateUI()

    MakeDungeonUICanvas(): canvas=
        Canvas := canvas:
            Slots:=array:
                #Main overlay
                canvas_slot:
                    Anchors:=anchors{Minimum:=vector2{X:=0.5,Y:=0.5},Maximum:=vector2{X:=0.5,Y:=0.5}}
                    ZOrder:=0
                    SizeToContent:=true
                    Offsets:=margin{Left:=0.0,Top:=0.0,Right:=0.0,Bottom:=0.0} #PositionX, PositionY, SizeX, SizeY
                    Alignment:=vector2{X:=0.5,Y:=0.5}
                    Widget:=texture_block{DefaultImage:=Images.BlackBackground,DefaultDesiredSize:=vector2{X:=BackgroundBlackUISizeX,Y:=BackgroundBlackUISizeY}}<#
                canvas_slot:
                    Anchors:=anchors{Minimum:=vector2{X:=0.5,Y:=0.5},Maximum:=vector2{X:=0.5,Y:=0.5}}
                    ZOrder:=0
                    SizeToContent:=true
                    Offsets:=margin{Left:=0.0,Top:=0.0,Right:=0.0,Bottom:=0.0} #PositionX, PositionY, SizeX, SizeY
                    Alignment:=vector2{X:=0.5,Y:=0.5}
                    Widget:=texture_block{DefaultImage:=Images.Img_MainBackground, DefaultDesiredSize:=vector2{X:=650.0,Y:=700.0}}#>
                canvas_slot:
                    Anchors:=anchors{Minimum:=vector2{X:=0.5,Y:=0.5},Maximum:=vector2{X:=0.5,Y:=0.5}}
                    ZOrder:=0
                    SizeToContent:=true
                    Offsets:=margin{Left:=0.0,Top:=0.0,Right:=0.0,Bottom:=0.0} #PositionX, PositionY, SizeX, SizeY
                    Alignment:=vector2{X:=0.5,Y:=0.5}
                    Widget:=overlay:
                        Slots:=array:
                            overlay_slot:
                                HorizontalAlignment:=horizontal_alignment.Fill
                                VerticalAlignment:=vertical_alignment.Fill
                                Widget:=texture_block{DefaultImage:=Images.Img_MainBackground, DefaultDesiredSize:=vector2{X:=650.0,Y:=700.0}} 
                            overlay_slot:
                                HorizontalAlignment:=horizontal_alignment.Fill
                                VerticalAlignment:=vertical_alignment.Fill
                                Padding:=margin{Top:=10.0,Bottom:=10.0,Right:=10.0,Left:=10.0}
                                Widget:=stack_box:  
                                    Orientation:=orientation.Vertical
                                        Slots:=array:
                                            stack_box_slot:
                                                Padding:=margin{Top:=5.0,Bottom:=5.0} 
                                                Widget:=Availability_TextBlock 
                                            stack_box_slot:
                                                Widget:=overlay:
                                                    Slots:=array: 
                                                        overlay_slot:
                                                            HorizontalAlignment:=horizontal_alignment.Right
                                                            VerticalAlignment:=vertical_alignment.Center
                                                            Widget:=overlay:
                                                                Slots:=array: 
                                                                    overlay_slot:
                                                                        Widget:=texture_block{DefaultDesiredSize:=vector2{X:=SizeBoxSaphiresX,Y:=SizeBoxSaphiresY},DefaultImage:=Images.EmptyTexture} 
                                                                    overlay_slot:
                                                                        HorizontalAlignment:=horizontal_alignment.Fill
                                                                        VerticalAlignment:=vertical_alignment.Fill
                                                                        Widget:=texture_block{DefaultImage:=Images.Img_LineBackground2}  
                                                                    overlay_slot:
                                                                        HorizontalAlignment:=horizontal_alignment.Left
                                                                        VerticalAlignment:=vertical_alignment.Center
                                                                        Padding:=margin{Left:=Icon_LeftMargin}
                                                                        Widget:=texture_block{DefaultDesiredSize:=vector2{X:=IconSmallSizeX,Y:=IconSmallSizeY},DefaultImage:=Images.Coal_Icon}
                                                                    overlay_slot:
                                                                        HorizontalAlignment:=horizontal_alignment.Left
                                                                        VerticalAlignment:=vertical_alignment.Center
                                                                        Padding:=margin{Left:=TextIcon_LeftMargin}
                                                                        Widget:=TotalCoal_TextBlock  
                                            stack_box_slot: 
                                                HorizontalAlignment:=horizontal_alignment.Center
                                                VerticalAlignment:=vertical_alignment.Center
                                                Padding:=margin{Top:=5.0,Bottom:=5.0} 
                                                Widget:=overlay:
                                                    Slots:=array:
                                                        overlay_slot:
                                                            Widget:=texture_block{DefaultImage:=Images.EmptyTexture,DefaultDesiredSize:=vector2{X:=650.0,Y:=100.0}}
                                                        overlay_slot:
                                                            Widget:=overlay:
                                                                Slots:=array:
                                                                    overlay_slot:
                                                                        Widget:=texture_block{DefaultImage:=Images.EmptyTexture,DefaultDesiredSize:=vector2{X:=650.0,Y:=80.0}}
                                                                    overlay_slot:
                                                                        HorizontalAlignment:=horizontal_alignment.Center
                                                                        Widget:=overlay:
                                                                            Slots:=array:
                                                                                overlay_slot:
                                                                                    HorizontalAlignment:=horizontal_alignment.Left
                                                                                    VerticalAlignment:=vertical_alignment.Center
                                                                                    Padding:=margin{Left:=Icon_LeftMargin}
                                                                                    Widget:=texture_block{DefaultDesiredSize:=vector2{X:=IconSmallSizeX,Y:=IconSmallSizeY},DefaultImage:=Images.Coal_Icon}
                                                                                overlay_slot:
                                                                                    HorizontalAlignment:=horizontal_alignment.Left
                                                                                    VerticalAlignment:=vertical_alignment.Center
                                                                                    Padding:=margin{Left:=TextIcon_LeftMargin}
                                                                                    Widget:=DungeonEvolve_TotalAgedSapphires_TextBlock
                                                                    overlay_slot:
                                                                        HorizontalAlignment:=horizontal_alignment.Fill
                                                                        VerticalAlignment:=vertical_alignment.Fill
                                                                        Widget:=ButtonEvolve
                                            stack_box_slot:
                                                Padding:=margin{Top:=5.0,Bottom:=5.0} 
                                                Widget:=Availability_TextBlock   
                                            stack_box_slot:
                                                Padding:=margin{Top:=5.0,Bottom:=5.0} 
                                                Widget:=NumPlayers_TextBlock
                                            stack_box_slot:
                                                Padding:=margin{Top:=5.0,Bottom:=5.0} 
                                                Widget:=Time_TextBlock
                                            stack_box_slot: 
                                                Widget:=DungeonImage
                                            stack_box_slot:
                                                Widget:=DungeonName_TextBlock
                                            stack_box_slot:
                                                Padding:=margin{Top:=5.0,Bottom:=5.0} 
                                                Widget:=stack_box:  
                                                    Orientation:=orientation.Horizontal
                                                    Slots:=array:
                                                        stack_box_slot: 
                                                            Padding:=margin{Left:=5.0,Right:=5.0}
                                                            Widget:=DungeonOriginalLV_TextBlock
                                                        stack_box_slot: 
                                                            Padding:=margin{Left:=5.0,Right:=5.0}
                                                            Widget:=DungeonLV_TextBlock
                                            stack_box_slot:
                                                Padding:=margin{Top:=5.0,Bottom:=5.0} 
                                                Widget:=stack_box:  
                                                    Orientation:=orientation.Horizontal
                                                    Slots:=array:
                                                        stack_box_slot: 
                                                            Padding:=margin{Left:=5.0,Right:=5.0}
                                                            Widget:=DungeonGatherInfo_Rewards
                                            stack_box_slot:
                                                Padding:=margin{Top:=5.0,Bottom:=5.0} 
                                                Widget:=stack_box:  
                                                    Orientation:=orientation.Horizontal
                                                    Slots:=array:
                                                        stack_box_slot: 
                                                            Padding:=margin{Left:=5.0,Right:=5.0}
                                                            Widget:=DungeonBigChest_Rewards
                                            stack_box_slot:
                                                Padding:=margin{Top:=5.0,Bottom:=5.0} 
                                                Widget:=stack_box:  
                                                    Orientation:=orientation.Horizontal
                                                    Slots:=array:
                                                        stack_box_slot: 
                                                            Padding:=margin{Left:=5.0,Right:=5.0}
                                                            Widget:=DungeonSmallChest_Rewards
                                            stack_box_slot:
                                                Padding:=margin{Top:=5.0,Bottom:=5.0} 
                                                HorizontalAlignment:=horizontal_alignment.Fill
                                                Widget:=overlay:
                                                    Slots:=array:
                                                        overlay_slot:
                                                            Widget:=texture_block{DefaultImage:=Images.EmptyTexture,DefaultDesiredSize:=vector2{X:=650.0,Y:=80.0}}
                                                        overlay_slot:
                                                            HorizontalAlignment:=horizontal_alignment.Fill
                                                            VerticalAlignment:=vertical_alignment.Fill
                                                            Widget:=ButtonPlay
                                            stack_box_slot:
                                                Padding:=margin{Top:=5.0,Bottom:=5.0} 
                                                HorizontalAlignment:=horizontal_alignment.Fill
                                                Widget:=overlay:
                                                    Slots:=array:
                                                        overlay_slot:
                                                            Widget:=texture_block{DefaultImage:=Images.EmptyTexture,DefaultDesiredSize:=vector2{X:=650.0,Y:=80.0}}
                                                        overlay_slot:
                                                            HorizontalAlignment:=horizontal_alignment.Fill
                                                            VerticalAlignment:=vertical_alignment.Fill
                                                            Widget:=ButtonExit
                                            stack_box_slot:
                                                Padding:=margin{Top:=5.0,Bottom:=5.0} 
                                                HorizontalAlignment:=horizontal_alignment.Fill
                                                Widget:=overlay:
                                                    Slots:=array:
                                                        overlay_slot:
                                                            Widget:=texture_block{DefaultImage:=Images.EmptyTexture,DefaultDesiredSize:=vector2{X:=650.0,Y:=80.0}}
                                                        overlay_slot:
                                                            HorizontalAlignment:=horizontal_alignment.Fill
                                                            VerticalAlignment:=vertical_alignment.Fill
                                                            Widget:=ButtonCancel

                            
MakeDungeonUI<constructor>(InBakeryManager : bakery_machines_manager
    )<transacts> := dungeon_ui :
        BakeryManager := InBakeryManager
                                                       

                            
           


