# fill this file with your verse-script
using { /UnrealEngine.com/Temporary/UI }
using { /UnrealEngine.com/Temporary/SpatialMath }
using { /Fortnite.com/UI }
using { /Verse.org/Simulation }
using { /Verse.org/Colors/NamedColors }

# A Verse-authored creative device that can be placed in a level
entity_health_bar_ui := class():

    #Class settings
    var MaybeAgent : ?agent = false

    BakeryMachinesManager : bakery_machines_manager

    #Estandar slider
    EstandarSlider : vector2 =vector2{X:=0.0,Y:=315.0}
    #HealthBar
    SliderUI : texture_block = texture_block{DefaultImage:=Images.EmptyTexture, DefaultDesiredSize := vector2{X:=0.0,Y:=0.0}}
    SliderTexture : texture_block = texture_block{DefaultImage:=Images.EmptyTexture, DefaultDesiredSize := vector2{X:=0.0,Y:=315.0}}
    Background : texture_block = texture_block{DefaultImage:=Images.HealthBackground, DefaultDesiredSize := vector2{X:=715.3,Y:=30.0}}
    Information_text : text_block = text_block{DefaultTextColor:=White, DefaultShadowOffset:=ShadowOffset, DefaultShadowOpacity := 1.0}
    Container_Ammount : texture_block = texture_block{DefaultImage:=Images.Slider_Container, DefaultDesiredSize := vector2{X:=ContainerHealth_HealthBarX,Y:=ContainerHealth_HealthBarY}}
    Difference_Slider : texture_block = texture_block{DefaultTint := Red, DefaultImage:=Images.Slider_Progress, DefaultDesiredSize := vector2{X:=0.0,Y:=35.0}}
    ActualAmmount_Slider : texture_block = texture_block{DefaultImage:=Images.Slider_Progress,DefaultDesiredSize := vector2{X:=0.0,Y:=35.0}}
    
    var SliderPercentage : float = 0.0

    var MaybeCanvas : ?canvas = false
    var UIIsDisplayed : logic = false

    var TimeElapsed : int = 5
    var TypeString : string= ""
    var Level : int = 0

    #Close
    CloseButton : button_loud = button_loud{}

    InitUI() : void =
        Canvas := MakeCanvas()
        set MaybeCanvas = option{Canvas}

    SetInfo(Ammount : float, Limit : float, Type : int, InLevel : int) : void =
        if(TimeElapsed >= 5):
            spawn:
                RunTimeElapsed()
        set TimeElapsed = 0
        #Info
        set Level = InLevel
        if(Type=HB_TYPE0):
            ActualAmmount_Slider.SetTint(WaterColor)
            Information_text.SetTextColor(WaterTextColor)
            set TypeString = "Water Stale Bread"
        else if(Type=HB_TYPE1):
            ActualAmmount_Slider.SetTint(FlourColor)
            Information_text.SetTextColor(FlourTextColor)
            set TypeString = "Flour Stale Bread"
        else if(Type=HB_TYPE2):
            ActualAmmount_Slider.SetTint(CheeseColor)
            Information_text.SetTextColor(CheeseTextColor)
            set TypeString = "Cheese Stale Bread"
        else if(Type=HB_TYPE3):
            ActualAmmount_Slider.SetTint(PimentoColor)
            Information_text.SetTextColor(PimentoTextColor)
            set TypeString = "Pimentos Stale Bread"
        else if(Type=HB_TYPE4):
            ActualAmmount_Slider.SetTint(MutantMaterialColor)
            Information_text.SetTextColor(MutantMaterialTextColor)
            set TypeString = "Unstable Stale Bread"
        else if(Type=ARROW_MINIGAME):
            ActualAmmount_Slider.SetTint(ArrowMiniGameColor)
            Information_text.SetTextColor(ArrowMiniGameTextColor)
            set TypeString = "Purge Valves"
        else if(Type=HB_PURE):
            ActualAmmount_Slider.SetTint(PureHBColor)
            Information_text.SetTextColor(PureHBTextColor)
            set TypeString = "Legendary"
        else if(Type=ENTITY_BOSS): #BOSS
            ActualAmmount_Slider.SetTint(MutantBreadColor)
            Information_text.SetTextColor(MutantBreadTextColor)
            set TypeString = "Mutant Bread Boss"
        SetHealth(Ammount,Limit,Type)
        

    ResetInfo() : void =
        XSliderValue := SliderPercentage * EntityHealth_MaxHealthBarX
        ActualAmmount_Slider.SetDesiredSize(vector2{X:=XSliderValue,Y:=35.0})
        Difference_Slider.SetDesiredSize(vector2{X:=XSliderValue,Y:=35.0})

    RunTimeElapsed()<suspends> : void =
        set TimeElapsed = 0
        SliderUI.SetDesiredSize(vector2{Y:=300.0})
        loop:
            Sleep(1.0)
            set TimeElapsed = TimeElapsed + 1
            Log("TimeElapsed: {TimeElapsed}",LOG_LEVEL_HEALTHBUG)
            if(TimeElapsed > 5):
                SliderUI.SetDesiredSize(vector2{Y:=0.0})
                ##Print("Reseted desired size")
                break

    DelayDifferenceBar<private>(StartPercentage : float, EndPercentage : float) <suspends> : void =
        spawn:
            SpawnDamageEffect()
        Difference_Slider.SetDesiredSize(vector2{X:=StartPercentage*EntityHealth_MaxHealthBarX,Y:=35.0})
        Sleep(0.35)
        Difference_Slider.SetDesiredSize(vector2{X:=EndPercentage*EntityHealth_MaxHealthBarX,Y:=35.0})

    SetHealth<private>(Ammount : float, Limit : float, Type : int) : void =
        NewSliderPercentage := Ammount / Limit
        ##Print("NewSliderPercentage: {NewSliderPercentage}")
        ##Print("Old Percentage: {SliderPercentage}")
        spawn:
            DelayDifferenceBar(SliderPercentage,NewSliderPercentage)
        set SliderPercentage = NewSliderPercentage
        XSliderValue := SliderPercentage * EntityHealth_MaxHealthBarX
        ActualAmmount_Slider.SetDesiredSize(vector2{X:=XSliderValue,Y:=35.0})
        ##Print("XSliderValue: {XSliderValue}")
        if(Type=HB_TYPE4):
            Information_text.SetText(BasicMessage("∞ - Unstable Slate Bread Lv: {Level}")) 
        else if(Type=HB_PURE):
            if(Level=HB_MAXLEVEL):
                Information_text.SetText(BasicMessage("Epic 96.2% Pure Slate Bread"))
        else:
            if(Level=0):
                if(Type<>HB_BASETYPE):
                    Information_text.SetText(BasicMessage("∞ -  {TypeString}")) 
                else:
                    Information_text.SetText(BasicMessage("∞ - Unpure Slate Bread")) 
                    Information_text.SetTextColor(White)
                ActualAmmount_Slider.SetDesiredSize(vector2{X:=EntityHealth_MaxHealthBarX,Y:=35.0})
            else:
                Information_text.SetText(BasicMessage("{GetOptimizedMagnitudeLetter(Ammount)} - {TypeString} Lv: {Level}")) 

    SetAgent(InAgent : agent) : void =
        set UIIsDisplayed = false
        set MaybeAgent = option{InAgent}
        

    SpawnDamageEffect()<suspends> : void =
        Background.SetTint(Red)
        Sleep(0.2)
        Background.SetTint(White)

    Remove_Canvas(Message : widget_message) : void =
        set UIIsDisplayed = false
        if:
            Agent := MaybeAgent?
            Player:=player[Agent]
            PlayerUI:=GetPlayerUI[Player]
            Canvas := MaybeCanvas?
        then:
            PlayerUI.RemoveWidget(Canvas)

    Remove_CanvasWithoutMessage() : void =
        set UIIsDisplayed = false
        if:
            Agent := MaybeAgent?
            Player:=player[Agent]
            PlayerUI:=GetPlayerUI[Player]
            Canvas := MaybeCanvas?
        then:
        #    #Print("Removing UI")
            PlayerUI.RemoveWidget(Canvas)

            
    Display_Canvas() : void =
        if:
            not UIIsDisplayed?
            Agent := MaybeAgent?
            PlayerUI:=GetPlayerUI[player[Agent]]
            Canvas := MaybeCanvas?
        then:
            set UIIsDisplayed = true
            PlayerUI.AddWidget(Canvas,player_ui_slot{ZOrder:=0})

    MakeCanvas<private>(): canvas=
        Canvas := canvas:
            Slots:=array:
                #Main overlay
                canvas_slot:
                    Anchors:=anchors{Minimum:=vector2{X:=0.5,Y:=0.0},Maximum:=vector2{X:=0.5,Y:=0.0}}
                    ZOrder:=0
                    SizeToContent:=true
                    Offsets:=margin{Left:=0.0,Top:=-150.0,Right:=0.0,Bottom:=0.0} #PositionX, PositionY, SizeX, SizeY
                    Alignment:=vector2{X:=0.5,Y:=0.0}
                    Widget:=overlay:
                        Slots:=array:
                            overlay_slot:      
                                Widget:=SliderUI
                            overlay_slot:
                                VerticalAlignment:=vertical_alignment.Bottom
                                Widget:=overlay:
                                    Slots:=array:
                                        overlay_slot:
                                            HorizontalAlignment:=horizontal_alignment.Left
                                            VerticalAlignment:=vertical_alignment.Center
                                            Padding:=margin{Left:=EntityHealth_PaddingSliderLeft,Bottom:=EntityHealth_PaddingSliderBottom}
                                            Widget:=Background
                                        overlay_slot:
                                            HorizontalAlignment:=horizontal_alignment.Left
                                            VerticalAlignment:=vertical_alignment.Center
                                            Padding:=margin{Left:=EntityHealth_PaddingSliderLeft,Bottom:=EntityHealth_PaddingSliderBottom}
                                            Widget:=Difference_Slider
                                        overlay_slot:
                                            HorizontalAlignment:=horizontal_alignment.Left
                                            VerticalAlignment:=vertical_alignment.Center
                                            Padding:=margin{Left:=EntityHealth_PaddingSliderLeft,Bottom:=EntityHealth_PaddingSliderBottom}
                                            Widget:=ActualAmmount_Slider
                                        overlay_slot:
                                            HorizontalAlignment:=horizontal_alignment.Center
                                            VerticalAlignment:=vertical_alignment.Center
                                            Widget:=Container_Ammount
                                        overlay_slot:
                                            Padding:=margin{Bottom:=EntityHealth_TextBottomPadding}
                                            HorizontalAlignment:=horizontal_alignment.Center
                                            VerticalAlignment:=vertical_alignment.Bottom
                                            Widget:=Information_text
MakeEntityHealthBarUI<constructor>(
    InBakeryMachinesManager : bakery_machines_manager)<transacts> := entity_health_bar_ui :
        BakeryMachinesManager := InBakeryMachinesManager
                                                                                                                