# fill this file with your verse-script
using { /UnrealEngine.com/Temporary/UI }
using { /UnrealEngine.com/Temporary/SpatialMath }
using { /Fortnite.com/UI }
using { /Verse.org/Simulation }
using { /Verse.org/Colors/NamedColors }

# A Verse-authored creative device that can be placed in a level
fast_travel_ui := class():

    #Class settings
    var BakeryMachines : bakery_machines_manager
    var MaybeAgent : ?agent = false

    var OpenedFromArea : int = -1

    TitleDescription : text_block = text_block{DefaultTextColor:=White, DefaultShadowOffset:=ShadowOffset, DefaultShadowOpacity := 1.0}

    #Area0
    Area0_Description : text_block = text_block{DefaultTextColor:=White, DefaultShadowOffset:=ShadowOffset, DefaultShadowOpacity := 1.0}
    Area0_Image : texture_block = texture_block{DefaultImage:=Images.TeleporterArea_0, DefaultDesiredSize := vector2{X:=300.0,Y:=200.0}}
    Area0_Button : button_regular = button_regular{}

    #Area1
    Area1_Description : text_block = text_block{DefaultTextColor:=White, DefaultShadowOffset:=ShadowOffset, DefaultShadowOpacity := 1.0}
    Area1_Image : texture_block = texture_block{DefaultImage:=Images.TeleporterArea_1, DefaultDesiredSize := vector2{X:=300.0,Y:=200.0}}
    Area1_Button : button_regular = button_regular{}

    #Area2
    Area2_Description : text_block = text_block{DefaultTextColor:=White, DefaultShadowOffset:=ShadowOffset, DefaultShadowOpacity := 1.0}
    Area2_Image : texture_block = texture_block{DefaultImage:=Images.TeleporterArea_1, DefaultDesiredSize := vector2{X:=300.0,Y:=200.0}}
    Area2_Button : button_regular = button_regular{}

    #Area3
    Area3_Description : text_block = text_block{DefaultTextColor:=White, DefaultShadowOffset:=ShadowOffset, DefaultShadowOpacity := 1.0}
    Area3_Image : texture_block = texture_block{DefaultImage:=Images.TeleporterArea_2, DefaultDesiredSize := vector2{X:=300.0,Y:=200.0}}
    Area3_Button : button_regular = button_regular{}

    #Area4
    Area4_Description : text_block = text_block{DefaultTextColor:=White, DefaultShadowOffset:=ShadowOffset, DefaultShadowOpacity := 1.0}
    Area4_Image : texture_block = texture_block{DefaultImage:=Images.TeleporterArea_3, DefaultDesiredSize := vector2{X:=300.0,Y:=200.0}}
    Area4_Button : button_regular = button_regular{}
    
    var MaybeCanvas : ?canvas = false
    var UIIsDisplayed : logic = false

    var TimeElapsed : int = 0

    #Close
    CloseButton : button_loud = button_loud{}

    var AllAreas_Description : []text_block = array{}
    var AllAreas_Images : []texture_block = array{}
    var AllAreas_Buttons : []button_regular = array{}

    SetAgent(InAgent : agent) : void =
        set MaybeAgent = option{InAgent}
        set UIIsDisplayed = false
        Remove_CanvasWithoutMessage()

    InitUI() : void =

        set AllAreas_Description = array{Area0_Description,Area1_Description,Area2_Description,Area3_Description,Area4_Description}
        set AllAreas_Images = array{Area0_Image,Area1_Image,Area2_Image,Area3_Image,Area4_Image}
        set AllAreas_Buttons =array{Area0_Button,Area1_Button,Area2_Button,Area3_Button,Area4_Button}

        Canvas := MakeCanvas()
        set MaybeCanvas = option{Canvas}

        CloseButton.SetText(BasicMessage("Cancel"))
        Area0_Button.SetText(BasicMessage("Travel free"))
        Area1_Button.SetText(BasicMessage("Travel free"))
        Area2_Button.SetText(BasicMessage("Travel free"))
        Area3_Button.SetText(BasicMessage("Travel free"))
        Area4_Button.SetText(BasicMessage("Travel free"))

        RemoveDelay_OnRemove()

        InitSubscribes()

    InitSubscribes() : void =
        CloseButton.OnClick().Subscribe(Remove_Canvas)
        Area0_Button.OnClick().Subscribe(TryTeleportToArea0)
        Area1_Button.OnClick().Subscribe(TryTeleportToArea1)
        Area2_Button.OnClick().Subscribe(TryTeleportToArea2)
        Area3_Button.OnClick().Subscribe(TryTeleportToArea3)
        Area4_Button.OnClick().Subscribe(TryTeleportToArea4)

    TryTeleportToArea0(WidgetMessage : widget_message) : void =
        TryTeleportToArea(0)
        BakeryMachines.CheckMusicPlay(BAKERSMUSIC_SOUND)

    TryTeleportToArea1(WidgetMessage : widget_message) : void =
        TryTeleportToArea(1)
        BakeryMachines.CheckMusicPlay(BAKERSMUSIC_SOUND)

    TryTeleportToArea2(WidgetMessage : widget_message) : void =
        TryTeleportToArea(2)
        BakeryMachines.CheckMusicPlay(BAKERSMUSIC_SOUND)

    TryTeleportToArea3(WidgetMessage : widget_message) : void =
        TryTeleportToArea(3)
        BakeryMachines.CheckMusicPlay(DESERTMUSIC_SOUND)

    TryTeleportToArea4(WidgetMessage : widget_message) : void =
        TryTeleportToArea(4)
        BakeryMachines.CheckMusicPlay(SECRETMUSIC_SOUND)

    TryTeleportToArea(AreaNum : int) : void =
        if(AgentOwner:=BakeryMachines.MaybeAgentOwner?):
            if(AreaNum=0):
                if(OpenedFromArea=TELEPORTER_FROM_BAKERY_AREA):
                    BakeryMachines.GameManager.TeleporterManager.TryToTeleportToLobby(AgentOwner)
                else if(OpenedFromArea=TELEPORTER_FROM_LOBBY_AREA):
                    BakeryMachines.GameManager.TeleporterManager.TryToFreeTeleportToBakery(AgentOwner)
                else if(OpenedFromArea=TELEPORTER_FROM_HEXAGAME_AREA):
                    BakeryMachines.GameManager.TeleporterManager.TryToTeleportToLobby(AgentOwner)
                else:
                    PrintFatalError("On TryTeleportToArea: OpenedFromArea: {OpenedFromArea} not valid")
            else if(AreaNum=1):
                if(OpenedFromArea=TELEPORTER_FROM_BAKERY_AREA):
                    BakeryMachines.GameManager.TeleporterManager.TryToTeleportToHexaGame(AgentOwner)
                else if(OpenedFromArea=TELEPORTER_FROM_LOBBY_AREA):
                    BakeryMachines.GameManager.TeleporterManager.TryToTeleportToHexaGame(AgentOwner)
                else if(OpenedFromArea=TELEPORTER_FROM_HEXAGAME_AREA):
                    BakeryMachines.GameManager.TeleporterManager.TryToFreeTeleportToBakery(AgentOwner)
                else:
                    PrintFatalError("On TryTeleportToArea: OpenedFromArea: {OpenedFromArea} not valid")
            else:
                if:
                    Player:=player[AgentOwner]
                    AreaLV:=BakeryMachines.GameManager.LVAreas_WithTeleport[AreaNum]
                    AreaLV<=BakeryMachines.PlayerDataCache.LV_Area
                then:
                    Log("Teleporter successfull to Area: {AreaNum}: Level Required : {AreaLV}, PlayerLevel: {BakeryMachines.PlayerDataCache.LV_Area}",LOG_LEVEL_DEBUG2)
                    BakeryMachines.GameManager.TeleporterManager.TryToTeleportPaidToArea(AreaNum,AgentOwner)  
                else:
                    BakeryMachines.GameManager.DisplayCharacterMessage(AgentOwner,H_CHARACTER,"Hey Buddy! You have not unlocked this area yet!",?Sound:=ERROR_SOUND)
            Remove_CanvasWithoutMessage()

    UpdateUI() : void =
        LVAreas_WithTeleport:=BakeryMachines.GameManager.LVAreas_WithTeleport
        NamesAreas_WithTeleport:=BakeryMachines.GameManager.AllNamesAreas_WithTeleport
        if(AgentOwner:=BakeryMachines.MaybeAgentOwner?,Player:=player[AgentOwner]):
                    if(OpenedFromArea=TELEPORTER_FROM_BAKERY_AREA):
                        Area0_Description.SetText(BasicMessage("LOBBY"))
                        Area0_Image.SetImage(Images.Teleporter_Lobby)
                        Area0_Button.SetText(BasicMessage("TRAVEL FREE"))
                        Area1_Description.SetText(BasicMessage("SOURDOUGH COMPETITION"))
                        Area1_Image.SetImage(Images.Teleporter_HexaGame)
                        Area1_Button.SetText(BasicMessage("TRAVEL FREE"))
                    else if(OpenedFromArea=TELEPORTER_FROM_LOBBY_AREA):
                        Area0_Description.SetText(BasicMessage("YOUR BAKERY"))
                        Area0_Image.SetImage(Images.Teleporter_Baker)
                        Area0_Button.SetText(BasicMessage("TRAVEL FREE"))
                        Area1_Description.SetText(BasicMessage("SOURDOUGH COMPETITION"))
                        Area1_Image.SetImage(Images.Teleporter_HexaGame)
                        Area1_Button.SetText(BasicMessage("TRAVEL FREE"))
                    else if(OpenedFromArea=TELEPORTER_FROM_HEXAGAME_AREA):
                        Area0_Description.SetText(BasicMessage("LOBBY"))
                        Area0_Image.SetImage(Images.Teleporter_Lobby)
                        Area0_Button.SetText(BasicMessage("TRAVEL FREE"))
                        Area1_Description.SetText(BasicMessage("YOUR BAKERY"))
                        Area1_Image.SetImage(Images.Teleporter_Baker)
                        Area1_Button.SetText(BasicMessage("TRAVEL FREE"))
            for(Index:=2..LVAreas_WithTeleport.Length-1):
                if:
                    AreaDescription:=AllAreas_Description[Index]
                    AreaImage:=AllAreas_Images[Index]
                    AreaButton:=AllAreas_Buttons[Index]
                    AreaName := NamesAreas_WithTeleport[Index]
                    LVRequired := LVAreas_WithTeleport[Index]
                then:
                    Log("Index UI {Index}, PlayerAreaLevel: {BakeryMachines.PlayerDataCache.LV_Area} AreaName: {AreaName}, LVRequired: {LVRequired}", LOG_LEVEL_DEBUG5)
                    if(LVRequired>BakeryMachines.PlayerDataCache.LV_Area):
                        AreaDescription.SetText(BasicMessage("LOCKED"))
                        AreaDescription.SetTextColor(Red)
                        AreaImage.SetTint(Red)
                        AreaButton.SetText(BasicMessage("LOCKED"))
                    else:
                        AreaDescription.SetText(BasicMessage(AreaName))
                        AreaDescription.SetTextColor(White)
                        AreaImage.SetTint(White)
                        AreaButton.SetText(BasicMessage("TRAVEL FREE"))


    Remove_Canvas(Message : widget_message) : void =
        set UIIsDisplayed = false
        if:
            AgentOwner:=BakeryMachines.MaybeAgentOwner?
            Player:=player[AgentOwner]
            PlayerUI:=GetPlayerUI[Player]
            Canvas := MaybeCanvas?
        then:
            PlayerUI.RemoveWidget(Canvas)
            RemoveDelay_OnRemove()

    Remove_CanvasWithoutMessage() : void =
        set UIIsDisplayed = false
        if:
            AgentOwner:=BakeryMachines.MaybeAgentOwner?
            Player:=player[AgentOwner]
            PlayerUI:=GetPlayerUI[Player]
            Canvas := MaybeCanvas?
        then:
            PlayerUI.RemoveWidget(Canvas)
            RemoveDelay_OnRemove()
  
    Display_Canvas(FromArea : int) : void =
        if(not UIIsDisplayed?):
            set OpenedFromArea = FromArea
            set TimeElapsed = 0
            set UIIsDisplayed = true
            TitleDescription.SetText(BasicMessage("FAST TRAVEL "))
            UpdateUI()
            if:
                AgentOwner:=BakeryMachines.MaybeAgentOwner?
                PlayerUI:=GetPlayerUI[player[AgentOwner]]
                Canvas := MaybeCanvas?
            then:
                PlayerUI.AddWidget(Canvas,player_ui_slot{InputMode := ui_input_mode.All, ZOrder:=1})
                RemoveDelay_OnAdd()
           #UpdateUI()

    RemoveDelay_OnAdd() : void =
        TitleDescription.SetText(BasicMessage("FAST TRAVEL"))

    RemoveDelay_OnRemove() : void =
        TitleDescription.SetText(BasicMessage("FAST TRAVEL "))

    MakeCanvas(): canvas=
        Canvas := canvas:
            Slots:=array:
                #Main overlay
                canvas_slot:
                    Anchors:=anchors{Minimum:=vector2{X:=0.5,Y:=0.5},Maximum:=vector2{X:=0.5,Y:=0.5}}
                    ZOrder:=0
                    SizeToContent:=true
                    Offsets:=margin{Left:=0.0,Top:=0.0,Right:=0.0,Bottom:=0.0} #PositionX, PositionY, SizeX, SizeY
                    Alignment:=vector2{X:=0.5,Y:=0.5}
                    Widget:=texture_block{DefaultImage:=Images.BlackBackground, DefaultDesiredSize:=vector2{X:=2500.0,Y:=2300.0}}
                canvas_slot:
                    Anchors:=anchors{Minimum:=vector2{X:=0.5,Y:=0.5},Maximum:=vector2{X:=0.5,Y:=0.5}}
                    ZOrder:=0
                    SizeToContent:=true
                    Offsets:=margin{Left:=0.0,Top:=0.0,Right:=0.0,Bottom:=0.0} #PositionX, PositionY, SizeX, SizeY
                    Alignment:=vector2{X:=0.5,Y:=0.5}
                    Widget:=overlay:
                        Slots:=array:
                            overlay_slot:
                                HorizontalAlignment:=horizontal_alignment.Fill
                                VerticalAlignment:=vertical_alignment.Fill
                                Padding:=margin{Left:=SelectorBoxPadding,Right:=SelectorBoxPadding,Top:=SelectorBoxPadding,Bottom:=SelectorBoxPadding}
                                Widget:=texture_block{DefaultImage:=Images.Img_MainBackground}
                            overlay_slot: #Screw 0
                                Padding:=margin{Left:=3.5,Top:=3.5,Right:=3.5,Bottom:=3.5}
                                HorizontalAlignment:=horizontal_alignment.Left
                                VerticalAlignment:=vertical_alignment.Top
                                Widget:=texture_block{DefaultImage:=Images.Dialogue_Bolt, DefaultDesiredSize:=vector2{X:=5.0,Y:=5.0}}
                            overlay_slot: #Screw 1
                                Padding:=margin{Left:=3.5,Top:=3.5,Right:=3.5,Bottom:=3.5}
                                HorizontalAlignment:=horizontal_alignment.Right
                                VerticalAlignment:=vertical_alignment.Top
                                Widget:=texture_block{DefaultImage:=Images.Dialogue_Bolt, DefaultDesiredSize:=vector2{X:=5.0,Y:=5.0}}
                            overlay_slot: #Screw 2
                                Padding:=margin{Left:=3.5,Top:=3.5,Right:=3.5,Bottom:=3.5}
                                HorizontalAlignment:=horizontal_alignment.Left
                                VerticalAlignment:=vertical_alignment.Bottom
                                Widget:=texture_block{DefaultImage:=Images.Dialogue_Bolt, DefaultDesiredSize:=vector2{X:=5.0,Y:=5.0}}
                            overlay_slot: #Screw 3
                                Padding:=margin{Left:=3.5,Top:=3.5,Right:=3.5,Bottom:=3.5}
                                HorizontalAlignment:=horizontal_alignment.Right
                                VerticalAlignment:=vertical_alignment.Bottom
                                Widget:=texture_block{DefaultImage:=Images.Dialogue_Bolt, DefaultDesiredSize:=vector2{X:=5.0,Y:=5.0}}
                            overlay_slot: 
                                Padding:=margin{Left:=10.0,Right:=10.0,Top:=10.0,Bottom:=10.0}
                                Widget:=stack_box:  
                                    Orientation:=orientation.Vertical
                                        Slots:=array:
                                            stack_box_slot:
                                                HorizontalAlignment:=horizontal_alignment.Center
                                                VerticalAlignment:=vertical_alignment.Center
                                                Widget:= TitleDescription
                                            stack_box_slot:
                                                HorizontalAlignment:=horizontal_alignment.Center
                                                VerticalAlignment:=vertical_alignment.Center
                                                Widget:= text_block{DefaultText:= BasicMessage("Fast travel from the factory is free!"), DefaultTextColor:=White, DefaultShadowOffset:=ShadowOffset, DefaultShadowOpacity := 1.0}  
                                            stack_box_slot:
                                                HorizontalAlignment:=horizontal_alignment.Center
                                                VerticalAlignment:=vertical_alignment.Center
                                                Widget:=stack_box:  
                                                    Orientation:=orientation.Horizontal
                                                        Slots:=array:
                                                            stack_box_slot: #AREA 0
                                                                Padding:=margin{Top:=10.0,Bottom:=10.0,Right:=10.0,Left:=10.0} 
                                                                Widget:=stack_box: 
                                                                    Orientation:=orientation.Vertical
                                                                        Slots:=array:
                                                                            stack_box_slot:
                                                                                Padding:=margin{Top:=5.0,Bottom:=5.0}
                                                                                Widget:=Area0_Image
                                                                            stack_box_slot:
                                                                                Padding:=margin{Top:=5.0,Bottom:=5.0}
                                                                                Widget:=Area0_Description
                                                                            stack_box_slot:
                                                                                Padding:=margin{Top:=5.0,Bottom:=5.0}
                                                                                Widget:=overlay:
                                                                                    Slots:=array:
                                                                                        overlay_slot:
                                                                                            Widget:=texture_block{DefaultImage:=Images.EmptyTexture, DefaultDesiredSize:=vector2{X:=350.0,Y:=LineSizeY}}
                                                                                        overlay_slot:
                                                                                            HorizontalAlignment:=horizontal_alignment.Fill
                                                                                            VerticalAlignment:=vertical_alignment.Fill
                                                                                            Widget:=Area0_Button
                                                            stack_box_slot: #Vertical line 
                                                                VerticalAlignment:=vertical_alignment.Fill
                                                                Widget:=texture_block{DefaultImage:=Images.Img_VerticalLine, DefaultDesiredSize:=vector2{X:=3.0}}
                                                            stack_box_slot: #AREA 1
                                                                Padding:=margin{Top:=10.0,Bottom:=10.0,Right:=10.0,Left:=10.0} 
                                                                Widget:=stack_box: 
                                                                    Orientation:=orientation.Vertical
                                                                        Slots:=array:
                                                                            stack_box_slot:
                                                                                Padding:=margin{Top:=5.0,Bottom:=5.0}
                                                                                Widget:=Area1_Image
                                                                            stack_box_slot:
                                                                                Padding:=margin{Top:=5.0,Bottom:=5.0}
                                                                                Widget:=Area1_Description
                                                                            stack_box_slot:
                                                                                Padding:=margin{Top:=5.0,Bottom:=5.0}
                                                                                Widget:=overlay:
                                                                                    Slots:=array:
                                                                                        overlay_slot:
                                                                                            Widget:=texture_block{DefaultImage:=Images.EmptyTexture, DefaultDesiredSize:=vector2{X:=350.0,Y:=LineSizeY}}
                                                                                        overlay_slot:
                                                                                            HorizontalAlignment:=horizontal_alignment.Fill
                                                                                            VerticalAlignment:=vertical_alignment.Fill
                                                                                            Widget:=Area1_Button
                                                            stack_box_slot: #Vertical line
                                                                VerticalAlignment:=vertical_alignment.Fill
                                                                Widget:=texture_block{DefaultImage:=Images.Img_VerticalLine, DefaultDesiredSize:=vector2{X:=3.0}}
                                                            stack_box_slot: #AREA 2
                                                                Padding:=margin{Top:=10.0,Bottom:=10.0,Right:=10.0,Left:=10.0} 
                                                                Widget:=stack_box: 
                                                                    Orientation:=orientation.Vertical
                                                                        Slots:=array:
                                                                            stack_box_slot:
                                                                                Padding:=margin{Top:=5.0,Bottom:=5.0}
                                                                                Widget:=Area2_Image
                                                                            stack_box_slot:
                                                                                Padding:=margin{Top:=5.0,Bottom:=5.0}
                                                                                Widget:=Area2_Description
                                                                            stack_box_slot:
                                                                                Padding:=margin{Top:=5.0,Bottom:=5.0}
                                                                                Widget:=overlay:
                                                                                    Slots:=array:
                                                                                        overlay_slot:
                                                                                            Widget:=texture_block{DefaultImage:=Images.EmptyTexture, DefaultDesiredSize:=vector2{X:=350.0,Y:=LineSizeY}}
                                                                                        overlay_slot:
                                                                                            HorizontalAlignment:=horizontal_alignment.Fill
                                                                                            VerticalAlignment:=vertical_alignment.Fill
                                                                                            Widget:=Area2_Button
                                                            stack_box_slot: #Vertical line 
                                                                VerticalAlignment:=vertical_alignment.Fill
                                                                Widget:=texture_block{DefaultImage:=Images.Img_VerticalLine, DefaultDesiredSize:=vector2{X:=3.0}}
                                                            stack_box_slot: #AREA 3
                                                                Padding:=margin{Top:=10.0,Bottom:=10.0,Right:=10.0,Left:=10.0} 
                                                                Widget:=stack_box: 
                                                                    Orientation:=orientation.Vertical
                                                                        Slots:=array:
                                                                            stack_box_slot:
                                                                                Padding:=margin{Top:=5.0,Bottom:=5.0}
                                                                                Widget:=Area3_Image
                                                                            stack_box_slot:
                                                                                Padding:=margin{Top:=5.0,Bottom:=5.0}
                                                                                Widget:=Area3_Description
                                                                            stack_box_slot:
                                                                                Padding:=margin{Top:=5.0,Bottom:=5.0}
                                                                                Widget:=overlay:
                                                                                    Slots:=array:
                                                                                        overlay_slot:
                                                                                            Widget:=texture_block{DefaultImage:=Images.EmptyTexture, DefaultDesiredSize:=vector2{X:=350.0,Y:=LineSizeY}}
                                                                                        overlay_slot:
                                                                                            HorizontalAlignment:=horizontal_alignment.Fill
                                                                                            VerticalAlignment:=vertical_alignment.Fill
                                                                                            Widget:=Area3_Button
                                                            stack_box_slot: #Vertical line 
                                                                VerticalAlignment:=vertical_alignment.Fill
                                                                Widget:=texture_block{DefaultImage:=Images.Img_VerticalLine, DefaultDesiredSize:=vector2{X:=3.0}}
                                                            stack_box_slot: #AREA 4
                                                                Padding:=margin{Top:=10.0,Bottom:=10.0,Right:=10.0,Left:=10.0} 
                                                                Widget:=stack_box: 
                                                                    Orientation:=orientation.Vertical
                                                                        Slots:=array:
                                                                            stack_box_slot:
                                                                                Padding:=margin{Top:=5.0,Bottom:=5.0}
                                                                                Widget:=Area4_Image
                                                                            stack_box_slot:
                                                                                Padding:=margin{Top:=5.0,Bottom:=5.0}
                                                                                Widget:=Area4_Description
                                                                            stack_box_slot:
                                                                                Padding:=margin{Top:=5.0,Bottom:=5.0}
                                                                                Widget:=overlay:
                                                                                    Slots:=array:
                                                                                        overlay_slot:
                                                                                            Widget:=texture_block{DefaultImage:=Images.EmptyTexture, DefaultDesiredSize:=vector2{X:=350.0,Y:=LineSizeY}}
                                                                                        overlay_slot:
                                                                                            HorizontalAlignment:=horizontal_alignment.Fill
                                                                                            VerticalAlignment:=vertical_alignment.Fill
                                                                                            Widget:=Area4_Button
                                            stack_box_slot:
                                                Padding:=margin{Top:=10.0,Bottom:=10.0,Right:=10.0,Left:=10.0} 
                                                Widget:=overlay:
                                                    Slots:=array:
                                                        overlay_slot: 
                                                            Widget:=texture_block{DefaultDesiredSize:=vector2{X:=1.0,Y:=BigCloseButtonSizeY},DefaultImage:=Images.EmptyTexture} 
                                                        overlay_slot:
                                                            HorizontalAlignment:=horizontal_alignment.Fill
                                                            VerticalAlignment:=vertical_alignment.Fill
                                                            Widget:= CloseButton

                                                


MakeFastTravelUI<constructor>(InBakeryManager : bakery_machines_manager)<transacts> := fast_travel_ui :
    BakeryMachines := InBakeryManager
                                                                                                                