
using { /Fortnite.com/Devices }
using { /Verse.org/Simulation }
using { /UnrealEngine.com/Temporary/Diagnostics }
using { /Verse.org/Random }
using { /Fortnite.com/Characters }
using { /UnrealEngine.com/Temporary/SpatialMath }

# See https://dev.epicgames.com/documentation/en-us/uefn/create-your-own-device-in-verse for how to create a verse device.

# A Verse-authored creative device that can be placed in a level
var MyVariable:weak_map(session,creative_device_base)= map{}

game_manager := class(creative_device):

    @editable
    AllBakeryCameras : []gameplay_camera_fixed_point_device = array{}

    @editable
    IsValveSimplifiedVersion : logic = true

    @editable 
    IslandSettingsDevice : experience_settings_device = experience_settings_device{}

    #UKTutorial
    @editable
    EnterUKTutorialButton : button_device = button_device{}
    @editable
    TutorialExitPosition : creative_prop = creative_prop{}
    @editable
    AllUKTutorials : []uk_tutorial = array{}
    @editable
    AllExtraCameras : []gameplay_camera_fixed_point_device = array{}

    #Language
    @editable
    English_MutatorsZone : mutator_zone_device = mutator_zone_device{}
    @editable
    Spanish_MutatorsZone : mutator_zone_device = mutator_zone_device{}
    @editable
    French_MutatorsZone : mutator_zone_device = mutator_zone_device{}
    @editable
    Portuguese_MutatorsZone : mutator_zone_device = mutator_zone_device{}

    #AnalyticsDevice
    @editable
    Zone15Analytic : analytics_device = analytics_device{}
    @editable
    Zone16Analytic : analytics_device = analytics_device{}
    @editable
    Zone17Analytic : analytics_device = analytics_device{}
    @editable
    Zone18Analytic : analytics_device = analytics_device{}
    @editable
    Zone19Analytic : analytics_device = analytics_device{}

    #Debug
    @editable
    DebugOn : logic = false
    @editable
    Button_DebugUI : button_device = button_device{}

    #Music
    @editable
    AllBakerMusic_MutatorsZones : []mutator_zone_device = array{}
    @editable
    AllDesertMusic_MutatorsZones : []mutator_zone_device = array{}
    @editable
    AllSecretMusic_MutatorsZones : []mutator_zone_device = array{}

    #HardBreadManager
    var AllHardBreads : []hardbread_managerv2 = array{}
    @editable
    AllWeakPoints_Manipulators : []prop_manipulator_device = array{}
    @editable
    AllWeakPoints_Props : []creative_prop = array{}
    var IsUsedWeakPoint : []logic = array{}
    var WeakPoint_OrTransform : []transform = array{}
    var AssignedHardBreads : []?hardbread_managerv2 = array{}


    #Areas
    @editable
    AllManualAreaCostMultipliers :  []float = array{
        1.0 #ZONE1
        1.0 #ZONE2
        1.0 #ZONE3
        1.0 #ZONE4
        1.0 #ZONE5
        1.0 #ZONE6
        1.0 #ZONE7
        1.0 #ZONE8
        1.0 #ZONE9
        1.0 #ZONE10
        1.0 #ZONE11
        1.0 #ZONE12
        1.0 #ZONE13
        1.0 #ZONE14
        1.0 #ZONE15
        1.0 #ZONE16
        1.0 #ZONE17
        1.0 #ZONE18
        1.0 #ZONE19
    }
    
    @editable
    LVAreas_WithTeleport : []int = array{}
    @editable
    AllNamesAreas_WithTeleport : []string = array{}
    @editable
    TeleporterManager : teleporter_manager = teleporter_manager{}

    @editable
    MiniGamesManager : minigames_manager = minigames_manager{}

    @editable
    CollectiblesManager : collectibles_manager = collectibles_manager{}

    #Help & music
    #@editable
    #SignalRemoteManager : signal_remote_manager_device = signal_remote_manager_device{}
    @editable
    InputTriggerOptions : input_trigger_device = input_trigger_device{}

    #PlayerSpanwers
    @editable
    AllPlayerSpawners : []player_spawner_device = array{}
    var AllAgentsJoinTime : [agent]float = map{}
    var RecentAgentsJoined : []agent = array{}

    #Class selectors
    @editable
    ClassTeamSelector_PickMode : class_and_team_selector_device =  class_and_team_selector_device{}
    @editable
    ClassTeamSelector_HexaGame : class_and_team_selector_device =  class_and_team_selector_device{}
    @editable
    ClassTeamSelector_Dungeon : class_and_team_selector_device =  class_and_team_selector_device{}
    @editable
    ClassTeamSelector_Tutorial : class_and_team_selector_device =  class_and_team_selector_device{}

    #Beacons
    @editable
    AllMapBeaconDevices : []beacon_device = array{}
    var AllMapBeaconDevices_IsUsed : []logic = array{}

    @editable
    GeneralMobsManager : general_mobs_manager = general_mobs_manager{}
    @editable
    AllHexaManager : []hexa_minigame_manager = array{}

    @editable
    AskForHelpMessage : hud_message_device = hud_message_device{}
    @editable
    QuestsAndOrdersDatabase : quests_and_orders_database = quests_and_orders_database{}
    
    #Resourcers Requirements
    # Manages and updates stats for each player.

    #Test
    var IsGivenPets : [agent]logic = map{}
    @editable
    Test_UpdateBackpackButton : button_device = button_device{}
    @editable
    Test_CompleteAllTutorialButton : button_device = button_device{}
    @editable
    Test_CompleteTutorialButton : button_device = button_device{}
    @editable
    Test_GrantAgedSapphires_Button : button_device = button_device{}
    @editable
    Test_GrantSapphires_Button : button_device = button_device{}
    @editable
    Test_GrantCoal_Button : button_device = button_device{}
    @editable
    Test_GrantMutantMaterial_Button : button_device = button_device{}
    @editable
    Test_GrantResourcers_Button : button_device = button_device{}
    @editable
    Test_LoadData_Button : button_device = button_device{}
    @editable
    Test_AddPhase_Button : button_device = button_device{}

    @editable
    Test_GrantOrderProgress_Button : button_device = button_device{}
    @editable
    Test_AddPermision_Button : button_device = button_device{}

    @editable
    Timer_InsertCoal : timer_device = timer_device{}
    @editable
    AllButtons_InserCoal : []button_device = array{}
    @editable
    Button_SkipQuest : button_device = button_device{}
    @editable
    Button_PetManager : button_device = button_device{}
    @editable
    Button_HexaGame : button_device = button_device{}
    @editable
    Button_ImproveHammer : button_device = button_device{}
    @editable
    Button_ImproveBow : button_device = button_device{}
    @editable
    Button_ImproveDungeonTokens : button_device = button_device{}
    @editable
    Button_ImproveHexaGameTokens : button_device = button_device{}

    #Quests
    @editable
    SkipQuestCoal_Cost : int = 15

    #Arrow
    @editable
    MaxArrows : int = 24
    @editable
    Arrows_ConditionalButton : conditional_button_device = conditional_button_device{}
    @editable
    Button_BuyArrows : button_device = button_device{}
    @editable
    ArrowsPrize : int = 5


    #Test
    var AllMobSpanwers : []npc_spawner_device = array{}

    @editable
    Button_CompleteReset : button_device = button_device{}
    @editable
    Button_Mastery : button_device = button_device{}

    @editable
    AllBarrierArea : []barrier_device = array{}
    @editable
    AllButtonsArea : []button_device = array{}
    @editable
    AllBillBoardsArea : []creative_prop = array{}
    #var AllAreaCosts : []float = array{}

    <#
    var AllAreaNames : []string = array{
        "Entrance to old oak city",
        "Chicken lovers restaurant",
        "City scrapyard",
        "Mr. White laundry",
        "Car wash",
        "Vacuum cleaner shop",
        "Entrance to the big desert"
        }#>

    PlayersDataManager : player_data_manager = player_data_manager{}
    PickaxeCalculator : pickaxe_calculator = pickaxe_calculator{}

    #HUDDevices
    @editable
    var HUDMessage_Personal : hud_message_device = hud_message_device{}

    #Super Energy / Coal
    @editable
    CoalGeneratorProp : creative_prop = creative_prop{}
    var MaybeCoalGeneratorProp_SuperEnergy : ?creative_prop = false
    var SuperEnergy_Shared : float = 0.0

    #Baker slots
    ###########
    var AllBakers : []?bakery_machines_manager = array{false,false,false,false}
    var AllAgents : []?agent = array{false,false,false,false}
    var AllAgentsBakery : [agent]bakery_machines_manager = map{}

    #Props Manager
    @editable
    PropsManager : props_manager = props_manager{}
    #Upgrades Manager
    #@editable
    #UpgradesManager : upgrades_manager = upgrades_manager{}
    #VFX Manager
    @editable
    VFXManager : vfx_and_accolades_manager = vfx_and_accolades_manager{}

    #AFK Zone
    var TotalGameTime : float = 0.0
    @editable
    MutZone_AfkZone : mutator_zone_device = mutator_zone_device{}
    var AllPlayersOnAFK_Zone : []player = array{}
    var AllPlayersAFKRegister : [agent]float = map{}

    #UpgradeButtons
    @editable
    Button_UpgradePickaxe : button_device = button_device{}
    @editable
    Button_UpgradePickaxePerks : button_device = button_device{}
    @editable
    Button_UpgradeBackpack : button_device = button_device{}

    
    var MaxNumHitsToWeakPoint  : int = 24

    EnterOnUKTutorial(Agent : agent) : void =
        AgentIndex:=GetAgentIndex(Agent)
        if(UKTutorial:=AllUKTutorials[AgentIndex]):
            UKTutorial.PlayUKTutorial(Agent)

    # Runs when the device is started in a running game
    OnBegin<override>()<suspends>:void=
        if(DebugOn?):
            PrintWarning("DEBUG IS ON!")
            for(Tier:=0..20):
                Log("Tier {Tier} degradation: {GetOptimizedMagnitudeLetter(Pow(PickaxeTierDegradation,Tier*1.0))}",LOG_LEVEL_DEBUG7)

        else:
            Test_GrantSapphires_Button.Disable()
            Test_AddPermision_Button.Disable()
            Test_GrantOrderProgress_Button.Disable()
            Test_CompleteTutorialButton.Disable()
        if(Quest_DelayTime_Base < 100.0 or Order_DelayTime_Base < 100.0):
            PrintLaunchWarning("Quest_DelayTime_Base is only {Quest_DelayTime_Base} and Order_DelayTime_Base is only {Order_DelayTime_Base}")

        
        if(Value:=Floor[((MinNum_AVGHit+MaxNum_AVGHit)/2.0)*PercentageToWeakPoint]):
            set MaxNumHitsToWeakPoint = Value
        InitSubscribers() #First thing to do!
        CollectiblesManager.SetGameManager(Self) #Must be here before PlayerAdded
        for(Player : GetPlayspace().GetPlayers()):
            OnPlayerAdded(Player)
        if(set MyVariable[GetSession()]=IslandSettingsDevice){Print("DEVICE HAS BEEN SET")}
        InitSpawnerSubscribers()
        #Game Loop
        InitLinkedDevices()
        InitAuxMapIndicator()
        Sleep(0.08)
        InitHardbread()
        Sleep(0.08)
        SetExtraStrings()
        Sleep(0.08)
        #PickaxeCalculator.GenerateHBHealthControllers()
        SetCostAreas()
        Sleep(0.08)
        HideVisualEmpty()
        Sleep(0.08)
        InitOthers()
        
        #RunBasicChecks()
        #ConvertDamageToSapphires(1.0)
        RunLoop()

    PlayEndTutorial(Agent : agent)<suspends> : void =
        AgentIndex:=GetAgentIndex(Agent)
        if(ValidBakery := GetAgentBakery_Transacted(Agent)?):
            if(Camera:=AllBakeryCameras[AgentIndex]):
                spawn:
                    PlayCamera(Camera,Agent,?CameraTime:=5.0)
            
            if(ValidBakery.PlayerDataCache.Language=ENGLISH):
                DisplayCharacterMessageLoc(Agent, I_CHARACTER, BasicMessage(otherstrings_en_module.ThisTheBakery), ?Sound:=MESSAGE_SOUND, ?DisplayTime:=7.0, ?Priority:=MAX_PRIORITY)
            else if(ValidBakery.PlayerDataCache.Language=SPANISH):
                DisplayCharacterMessageLoc(Agent, I_CHARACTER, BasicMessage(otherstrings_es_module.ThisTheBakery), ?Sound:=MESSAGE_SOUND, ?DisplayTime:=7.0, ?Priority:=MAX_PRIORITY)
            else if(ValidBakery.PlayerDataCache.Language=FRENCH):
                DisplayCharacterMessageLoc(Agent, I_CHARACTER, BasicMessage(otherstrings_fr_module.ThisTheBakery), ?Sound:=MESSAGE_SOUND, ?DisplayTime:=7.0, ?Priority:=MAX_PRIORITY)
            else if(ValidBakery.PlayerDataCache.Language=PORTUGUESE):
                DisplayCharacterMessageLoc(Agent, I_CHARACTER, BasicMessage(otherstrings_pt_module.ThisTheBakery), ?Sound:=MESSAGE_SOUND, ?DisplayTime:=7.0, ?Priority:=MAX_PRIORITY)
            
            Sleep(7.5)
            
            if(Camera:=AllExtraCameras[2]):
                spawn:
                    PlayCamera(Camera,Agent,?CameraTime:=5.0)
            
            if(ValidBakery.PlayerDataCache.Language=ENGLISH):
                DisplayCharacterMessageLoc(Agent, I_CHARACTER, BasicMessage(otherstrings_en_module.Shop), ?Sound:=MESSAGE_SOUND, ?DisplayTime:=7.0, ?Priority:=MAX_PRIORITY)
            else if(ValidBakery.PlayerDataCache.Language=SPANISH):
                DisplayCharacterMessageLoc(Agent, I_CHARACTER, BasicMessage(otherstrings_es_module.Shop), ?Sound:=MESSAGE_SOUND, ?DisplayTime:=7.0, ?Priority:=MAX_PRIORITY)
            else if(ValidBakery.PlayerDataCache.Language=FRENCH):
                DisplayCharacterMessageLoc(Agent, I_CHARACTER, BasicMessage(otherstrings_fr_module.Shop), ?Sound:=MESSAGE_SOUND, ?DisplayTime:=7.0, ?Priority:=MAX_PRIORITY)
            else if(ValidBakery.PlayerDataCache.Language=PORTUGUESE):
                DisplayCharacterMessageLoc(Agent, I_CHARACTER, BasicMessage(otherstrings_pt_module.Shop), ?Sound:=MESSAGE_SOUND, ?DisplayTime:=7.0, ?Priority:=MAX_PRIORITY)
            
            Sleep(7.5)
            
            if(ValidBakery.PlayerDataCache.Language=ENGLISH):
                DisplayCharacterMessageLoc(Agent, I_CHARACTER, BasicMessage(otherstrings_en_module.Pets), ?Sound:=MESSAGE_SOUND, ?DisplayTime:=7.0, ?Priority:=MAX_PRIORITY)
            else if(ValidBakery.PlayerDataCache.Language=SPANISH):
                DisplayCharacterMessageLoc(Agent, I_CHARACTER, BasicMessage(otherstrings_es_module.Pets), ?Sound:=MESSAGE_SOUND, ?DisplayTime:=7.0, ?Priority:=MAX_PRIORITY)
            else if(ValidBakery.PlayerDataCache.Language=FRENCH):
                DisplayCharacterMessageLoc(Agent, I_CHARACTER, BasicMessage(otherstrings_fr_module.Pets), ?Sound:=MESSAGE_SOUND, ?DisplayTime:=7.0, ?Priority:=MAX_PRIORITY)
            else if(ValidBakery.PlayerDataCache.Language=PORTUGUESE):
                DisplayCharacterMessageLoc(Agent, I_CHARACTER, BasicMessage(otherstrings_pt_module.Pets), ?Sound:=MESSAGE_SOUND, ?DisplayTime:=7.0, ?Priority:=MAX_PRIORITY)
            
            if(Camera:=AllExtraCameras[3]):
                spawn:
                    PlayCamera(Camera,Agent,?CameraTime:=5.0)
            
            Sleep(2.0)
            
            if(Camera:=AllExtraCameras[4]):
                spawn:
                    PlayCamera(Camera,Agent,?CameraTime:=3.0)
            
            ValidBakery.RefreshAllUI()
            
            Sleep(7.5)
            
            if(Camera:=AllExtraCameras[5]):
                spawn:
                    PlayCamera(Camera,Agent,?CameraTime:=5.0)
            
            if(ValidBakery.PlayerDataCache.Language=ENGLISH):
                DisplayCharacterMessageLoc(Agent, I_CHARACTER, BasicMessage(otherstrings_en_module.CompleteAgain), ?Sound:=MESSAGE_SOUND, ?DisplayTime:=7.0, ?Priority:=MAX_PRIORITY)
            else if(ValidBakery.PlayerDataCache.Language=SPANISH):
                DisplayCharacterMessageLoc(Agent, I_CHARACTER, BasicMessage(otherstrings_es_module.CompleteAgain), ?Sound:=MESSAGE_SOUND, ?DisplayTime:=7.0, ?Priority:=MAX_PRIORITY)
            else if(ValidBakery.PlayerDataCache.Language=FRENCH):
                DisplayCharacterMessageLoc(Agent, I_CHARACTER, BasicMessage(otherstrings_fr_module.CompleteAgain), ?Sound:=MESSAGE_SOUND, ?DisplayTime:=7.0, ?Priority:=MAX_PRIORITY)
            else if(ValidBakery.PlayerDataCache.Language=PORTUGUESE):
                DisplayCharacterMessageLoc(Agent, I_CHARACTER, BasicMessage(otherstrings_pt_module.CompleteAgain), ?Sound:=MESSAGE_SOUND, ?DisplayTime:=7.0, ?Priority:=MAX_PRIORITY)
            
            Sleep(7.5)
            
            if(ValidBakery.PlayerDataCache.Language=ENGLISH):
                DisplayCharacterMessageLoc(Agent, I_CHARACTER, BasicMessage(otherstrings_en_module.GoodLuck), ?Sound:=MESSAGE_SOUND, ?DisplayTime:=7.0, ?Priority:=MAX_PRIORITY)
            else if(ValidBakery.PlayerDataCache.Language=SPANISH):
                DisplayCharacterMessageLoc(Agent, I_CHARACTER, BasicMessage(otherstrings_es_module.GoodLuck), ?Sound:=MESSAGE_SOUND, ?DisplayTime:=7.0, ?Priority:=MAX_PRIORITY)
            else if(ValidBakery.PlayerDataCache.Language=FRENCH):
                DisplayCharacterMessageLoc(Agent, I_CHARACTER, BasicMessage(otherstrings_fr_module.GoodLuck), ?Sound:=MESSAGE_SOUND, ?DisplayTime:=7.0, ?Priority:=MAX_PRIORITY)
            else if(ValidBakery.PlayerDataCache.Language=PORTUGUESE):
                DisplayCharacterMessageLoc(Agent, I_CHARACTER, BasicMessage(otherstrings_pt_module.GoodLuck), ?Sound:=MESSAGE_SOUND, ?DisplayTime:=7.0, ?Priority:=MAX_PRIORITY)
            
            Sleep(7.5)
            
            ClassTeamSelector_PickMode.ChangeClass(Agent)
        if(Player:=player[Agent], AgentBakery:=GetAgentBakery_Transacted(Agent)?):
            spawn:
                QuestsAndOrdersDatabase.DelayCheckWelcomeLore(Player,AgentBakery,2.0)

    CalculateAndSaveFame(Agent : agent)<suspends> : void =
        if(AgentBakery:=GetAgentBakery_Transacted(Agent)?):
            OldFame := AgentBakery.PlayerDataCache.Inventory_Fame

            var TotalCoalInvested : float = 0.0
            var TotalMutantMaterialInvested : float = 0.0
            TotalCompletedChallenges := AgentBakery.PlayerDataCache.HexaGame_AllChallengesCompleted.Length

            for(Pet : AgentBakery.PlayerDataCache.AllPets):
                TupleAmountInvested:=CalculePetTotalAmountsInvested(Pet)
                set TotalMutantMaterialInvested =  TotalMutantMaterialInvested + TupleAmountInvested(0)
                set TotalCoalInvested =  TotalCoalInvested + TupleAmountInvested(1)
                set TotalCoalInvested =  TotalCoalInvested + TupleAmountInvested(2)

            NewFame := (TotalMutantMaterialInvested + (TotalCoalInvested*MutantMaterialPerCoal)) * TotalCompletedChallenges
            Difference := NewFame - OldFame

            if(Difference<>0.0):
                AgentBakery.SetAsyncNewFame(NewFame,Difference)
        
    InitOthers() : void =
        Arrows_ConditionalButton.SetItemCountRequired(0,MaxArrows)
        Button_SkipQuest.SetInteractionText(BasicMessage("Complete your current Quest ({SkipQuestCoal_Cost} Coal) / Skip Tutorial (Free)"))

    DebugAllUIs(Agent : agent)<suspends> : void =
        if(AgentBakery := GetAgentBakery_Transacted(Agent)?):
            AgentBakery.CloseAllUI()
            Sleep(1.00)
            #AgentBakery.DisplayBasicUIs()
            Sleep(0.08)
            DisplayCharacterMessage(Agent, G_CHARACTER, "UI Deleted", ?Sound:=SMALLOBJETIVE_SOUND)

    HandlerDebugAllUIs(Agent : agent) : void =
        spawn:
            DebugAllUIs(Agent)

    HideVisualEmpty<private>() : void = 
        AllObjects:=FindCreativeObjectsWithTag(visualempty{})
        for(Object : AllObjects, VisualEmpty:= creative_prop[Object]):
            VisualEmpty.SetMaterial(Materials.BP_Nothing)
            NewPosition : vector3 = vector3{X:=VisualEmpty.GetTransform().Translation.X,Y:=VisualEmpty.GetTransform().Translation.Y,Z:=VisualEmpty.GetTransform().Translation.Z-VisualEmptyHideZ}
            if(VisualEmpty.TeleportTo[NewPosition,VisualEmpty.GetTransform().Rotation]):

    WeakPointEvent(Agent : agent, ManipulatorIndex : int) : void =
        ##Print("Weak point detected on weak Index: {ManipulatorIndex}")
        if(HardBreadV2:=AssignedHardBreads[ManipulatorIndex]?, Manipulator := AllWeakPoints_Manipulators[ManipulatorIndex]):
            VFXManager.Accolade_WeakPoint.Award(Agent)
            ##Print("Hardbread detected")
            HardBreadV2.WeakPointEvent(Agent)
            Manipulator.RestockResources()

    FindWeakPointAuxiliar(HardBreadManager : hardbread_managerv2) : int = 
        var Index : int = -1
        loop:
            if(IsUsed:=IsUsedWeakPoint[Index], not IsUsed?, set IsUsedWeakPoint[Index]=true, set AssignedHardBreads[Index] = option{HardBreadManager}):
                return Index
            else:
                set Index = Index + 1
            if(Index>IsUsedWeakPoint.Length-1):
                #Print("Fatal error, Index>IsUsedWeakPoint.Length = {Index}>{IsUsedWeakPoint.Length}")
                return -1

    FreeWeakPoint(HBTransform : transform, Index : int)<suspends> : void =
        if(OriginalTransform := WeakPoint_OrTransform[Index], WeakProp := AllWeakPoints_Props[Index]):
            WeakProp.Hide()
            #NewPosition : vector3 = vector3{X:=HBTransform.Translation.X,Y:=HBTransform.Translation.Y,Z:=HBTransform.Translation.Z-16.0}
            #WeakProp.MoveTo(NewPosition,HBTransform.Rotation,0.5)
            if(WeakProp.TeleportTo[OriginalTransform]):
            else:
                TeleportFailedAttempt(false, false, option{WeakProp}, OriginalTransform, 256.0)
            Sleep(2.0)
            WeakProp.Show()
                Log("Free weak point {Index}", LOG_LEVEL_HB2)
            if(set IsUsedWeakPoint[Index] = false):


    PlaceWeakPoint(HardBreadManager : hardbread_managerv2, WeakPointAuxiliar_Index : int, ?IsLastLevel : logic = false) : logic =
        var IsSuccess : logic = false
        var ForwardsMovement : float = 0.0
        if(IsLastLevel?):
            set ForwardsMovement = 256.0
        else:
            set ForwardsMovement = 128.0

        CentralTransform:=HardBreadManager.GetCentralWeakTransform(IsLastLevel)
        var RandomRotation : rotation = CentralTransform.Rotation.ApplyYaw(DegreesToRadians(GetRandomFloat(0.0,359.0)))
        set RandomRotation = CentralTransform.Rotation.ApplyPitch(DegreesToRadians(GetRandomFloat(-0.0,195.0)))

        PointingVector:=RandomRotation.GetLocalForward()*ForwardsMovement #Old 112
        NewPosition:=CentralTransform.Translation+PointingVector
        NewTransform:=transform{Translation:=NewPosition,Rotation:=RandomRotation,Scale:=NeutralScale}
        if(WeakProp := AllWeakPoints_Props[WeakPointAuxiliar_Index]):
            if:
                WeakPointAuxiliar_Index<>-1
                WeakProp.TeleportTo[NewPosition,RandomRotation]
            then:   
                Log("Teleported weak prop {WeakPointAuxiliar_Index} to position X:={NewPosition.X} Y:={NewPosition.Y} Z:={NewPosition.Z}", LOG_LEVEL_HB2)
                return true
            else:
                Result:=TeleportFailedAttempt(false, false, option{WeakProp}, NewTransform, 15.0)(0)
                PrintFatalError("Fatal error teleporting weak prop")
                return Result
        else:
            PrintWarning("WeakProp {WeakPointAuxiliar_Index} not found!")
            return false
        #RandomRotation found, now apply the forward

        #Roll is X: Red Axis OF THE OBJECT
        #Pitch is Y: Green Axis OF THE OBJECT
        #Yaw is Z: Blue axis OF THE OBJECT
        

    #AFK ZONE
    AgentEntersAFKZone(Agent : agent) : void =
        if(ValidBakery := GetAgentBakery_Transacted(Agent)?):
            ValidBakery.ClearAFKRegister()
        if(Player:=player[Agent]):
            if(set AllPlayersAFKRegister[Agent]=TotalGameTime):
            set AllPlayersOnAFK_Zone = AllPlayersOnAFK_Zone + array{Player}

    AgentExitsAFKZone(Agent : agent) : void =
        var Index : int = 0
        if(ValidBakery := GetAgentBakery_Transacted(Agent)?, Player := player[Agent]):
            if(EnterTime:= AllPlayersAFKRegister[Agent]):
                TimeOnAFKZone := TotalGameTime-EnterTime
                if(IntTime:=Ceil[TimeOnAFKZone]):
                    DisplayCharacterMessage(Agent, I_CHARACTER, "Rested {IntTime}s, collected {GetOptimizedMagnitudeLetter(ValidBakery.AFKRegister)} Mutant Material", ?Sound:=SMALLOBJETIVE_SOUND)
            loop:
                if(OtherPlayer:=AllPlayersOnAFK_Zone[Index], OtherPlayer = Player, NewArray := AllPlayersOnAFK_Zone.RemoveElement[Index]):
                    set AllPlayersOnAFK_Zone = NewArray
                    break
                if(AllPlayersOnAFK_Zone.Length<Index):
                    break
                else: 
                    set Index = Index + 1
    Display_OptionsMenu(Agent : agent) : void =
        MaybeBakery:=GetAgentBakery(Agent)
        if(ValidBakery := MaybeBakery?):
            ValidBakery.Display_OptionsMenu()

    SetExtraStrings() : void =
        Button_BuyArrows.SetInteractionText(BasicMessage("Buy 5 Arrows: {ArrowsPrize} Coal"))

    OnPlayerSpawn(InAgent : agent) : void =
        var NewAgentDetected : logic = false
        ##Print("A player just spawned!")    
        for(Index:=0..RecentAgentsJoined.Length-1):
            if(RecentAgentsJoined[Index] = InAgent):
                if(Player := player[InAgent], Player.IsActive[]):
                    OnPlayerAdded(Player)
                set NewAgentDetected = true
                if(NewRecentAgentsJoined := RecentAgentsJoined.RemoveElement[Index], set RecentAgentsJoined = NewRecentAgentsJoined):
        ClassTeamSelector_PickMode.ChangeClass(InAgent)
        MaybeBakery:=GetAgentBakery(InAgent)
        if(ValidBakery:=MaybeBakery?):
            ValidBakery.PetsManager.Generate_PlayerEnhacements()
            ValidBakery.CheckMusicPlay(BAKERSMUSIC_SOUND)
            if(ValidBakery.IsOnUKTutorial?, UKTutorial := ValidBakery.MaybeUKTutorial?):
                UKTutorial.TeleportToCurrentPhase(InAgent)
            spawn:
                GeneralMobsManager.GranAgentWeapons(InAgent)

    LateUIRefresh(ValidBakery : bakery_machines_manager)<suspends> : void =
        Sleep(3.0)
        ValidBakery.RefreshAllUI()

    RefreshInventoryUI(Agent : agent) : void = 
        MaybeBakery:=GetAgentBakery(Agent)
        if(ValidBakery:=MaybeBakery?):
            ValidBakery.RefreshInventoryUI(Agent)

    TryBuyArrows(Agent : agent) : void =
        FloatArrowsPrize:=ArrowsPrize*1.0
        if(Player:=player[Agent], Player.IsActive[]<#, PlayerData:=PlayersDataManager.GetPlayerData[Player]#>):
            MaybeBakery:=GetAgentBakery(Agent)
            if(ValidBakery:=MaybeBakery?):
                if(not Arrows_ConditionalButton.HasAllItems[Agent]):
                    if(ValidBakery.PlayerDataCache.Inventory_Coal>=FloatArrowsPrize):
                        GeneralMobsManager.WeaponsGranter.SetNextItem(5)
                        GeneralMobsManager.WeaponsGranter.GrantItem(Agent)
                        var NewData : player_data_table = PlayersDataManager.WithdrawResource(ValidBakery.PlayerDataCache, FloatArrowsPrize, COAL)
                        ValidBakery.DisplayResourcersGather(array{-FloatArrowsPrize},array{COAL},false)
                        ValidBakery.SetNewPlayerDataCache(NewData)
                        #PlayersDataManager.SaveAllNewData(Player,PlayerData,NewData,ValidBakery)
                        
                        VFXManager.ArrowGranted_AudioPlayer.Play(Agent)
                        #VFXManager.PlayerPays_AudioPlayer.Play(Agent)
                        RefreshInventoryUI(Agent)
                        spawn:
                            DelaySaveNumArrows(Agent)
                    else if(AmountMissing:=Floor[FloatArrowsPrize-ValidBakery.PlayerDataCache.Inventory_Coal]):
                        DisplayCharacterMessage(Agent, I_CHARACTER,"You need {AmountMissing} Coal", ?Sound:=ERROR_SOUND)
                        VFXManager.ErrorSound_AudioPlayer.Play(Agent)
                else:
                    DisplayCharacterMessage(Agent, I_CHARACTER, "You already have {MaxArrows} arrows at max capacity", ?Sound:=ERROR_SOUND) 
                    VFXManager.ErrorSound_AudioPlayer.Play(Agent)
                    NewData := PlayersDataManager.SetAmmoArrow(ValidBakery.PlayerDataCache,MaxArrows)
                    ValidBakery.SetNewPlayerDataCache(NewData)
            else:
                PrintFatalError("On TryBuyArrows: bakery not found!")

    DelaySaveNumArrows(Agent : agent)<suspends> : void =
        Sleep(0.1)
        if(Player:=player[Agent], Player.IsActive[], ValidBakery := GetAgentBakery_Transacted(Agent)?):  
            AmountArrows:=GeneralMobsManager.HowManyArros_ConditionalButton.GetItemCount(Agent,0)
            Log("Num arrows on inventory: {AmountArrows}",LOG_LEVEL_DEBUG7)
            ValidBakery.SetNewPlayerDataCache(PlayersDataManager.SetAmmoArrow(ValidBakery.PlayerDataCache,AmountArrows))

    ChangeAgentClass_ToPickMode(Agent : agent) : void =
        ClassTeamSelector_PickMode.ChangeClass(Agent)
        if(ValidBakery:=GetAgentBakery_Transacted(Agent)?):
            ValidBakery.PetsManager.Generate_PlayerEnhacements()
    ChangeAgentClass_ToDungeonMode(Agent : agent) : void =
        ClassTeamSelector_Dungeon.ChangeClass(Agent)
        if(ValidBakery:=GetAgentBakery_Transacted(Agent)?):
            ValidBakery.PetsManager.Generate_PlayerEnhacements()
    ChangeAgentClass_ToHexaGameMode(Agent : agent) : void =
        ClassTeamSelector_HexaGame.ChangeClass(Agent)
        if(ValidBakery:=GetAgentBakery_Transacted(Agent)?):
            ValidBakery.PetsManager.Generate_PlayerEnhacements()

    PlayHexaGame(Agent : agent) : void =
        MaybeBakery:=GetAgentBakery(Agent)
        if(ValidBakery:=MaybeBakery?):
            ValidBakery.HexaGameManager.ShowChallenges(Agent)
        else:
            PrintFatalError("Bakery of player not found!")

    InitAuxMapIndicator() : void =
        for(Index:=0..AllMapBeaconDevices.Length-1):
            set AllMapBeaconDevices_IsUsed = AllMapBeaconDevices_IsUsed + array{false}

    DisplayResourcersGather(Agent : agent, Amounts : []float, Types : []int, IsCritical : logic) : void =
        MaybeBakery:=GetAgentBakery(Agent)
        if(ValidBakery:=MaybeBakery?):
            ValidBakery.DisplayResourcersGather(Amounts,Types,IsCritical)

    InitLinkedDevices() : void =
        MiniGamesManager.Init(Self)
        TeleporterManager.SetGameManager(Self)
        QuestsAndOrdersDatabase.SetGameManager(Self)
        GeneralMobsManager.SetGameManager(Self)
        GeneralMobsManager.DungeonDataBase.SetGameManager(Self)
        PickaxeCalculator.GenerateHBHealthControllers(Self)

    UpdateAll_DungeonUI(NumDungeon : int) : void =
        for(MaybeBaker : AllBakers, ValidBaker := MaybeBaker?):
            ValidBaker.UpdateDungeonUI(NumDungeon)

    RemoveQuest_MapTracker(Agent : agent) : void =
        for(BeaconDevice : AllMapBeaconDevices):
            BeaconDevice.RemoveFromShowList(Agent)

    AssignQuest_MapTracker(Agent : agent, NumQuestsCompleted : int, QuestType : int) : void =
        IndexToUse:=GetAgentIndex(Agent)
        MaybeBakery:=GetAgentBakery(Agent)
        if(ValidBakery:=MaybeBakery?):
            TupleResult:=QuestsAndOrdersDatabase.GetQuestObjetiveTransform(<#NumQuest#>ValidBakery.PlayerDataCache.NumQuestsCompleted,QuestType,Agent,ValidBakery.PlayerDataCache,ValidBakery.PlayerDataCache.QuestAssigned_NumQuest)
            IsValid:=TupleResult(0)
            TransformToTeleport:=TupleResult(1)
            #MaybeButton:=QuestsAndOrdersDatabase.GetButtonDevice(QuestType,NumQuest,Agent)
            if(IsValid?, MapBeaconDevice:=AllMapBeaconDevices[IndexToUse]):
                #var TransformToTeleport : transform = transform{}
                #if(Prop:=MaybeProp?):
                #    set TransformToTeleport = Prop.GetTransform()
                #else if(ButtonDevice:=MaybeButton?):
                #    set TransformToTeleport = ButtonDevice.GetTransform()
                MapBeaconDevice.AddToShowList(Agent)
                if(MapBeaconDevice.TeleportTo[TransformToTeleport]):
                    #Log("Teleported to X: {TransformToTeleport.Translation.X}, Y: {TransformToTeleport.Translation.Y}, Z: {TransformToTeleport.Translation.Z}", LOG_LEVEL_DEBUG7)
                    if(ValidBakery.PlayerDataCache.IsQuestTrackerOn?):
                        VFXManager.SpawnQuestTracker(Agent,TransformToTeleport,IndexToUse,TotalGameTime)
                    
                    ##Print("Teleporter succesfull")
                else:
                    PrintFatalError("AssignQuest_MapTracker")

    Test_AddProgressToOrder(Agent : agent) : void =
        if(Player := player[Agent], Player.IsActive[]):
            MaybeBakery:=GetAgentBakery(Agent)
            if(Bakery:=MaybeBakery?, MainInventoryUI := Bakery.MaybeMainInventoryUI?, ActionUI := Bakery.MaybeActionInfoUI?):
                #Amount calculated to add test
                AmountToAdd:=Bakery.PlayerDataCache.OrderAssigned_ProductAmountRequested/3.0
                OrderInfoResult:=QuestsAndOrdersDatabase.AddAgentOrder_Progress(Player,Bakery.PlayerDataCache,AmountToAdd,AmountToAdd,AmountToAdd)
                if(OrderInfoResult(0)?):
                    #Print("OrderInfoResult(2) is {OrderInfoResult(2)}, OrderInfoResult(3) is {OrderInfoResult(3)}")
                    ActionUI.DisplayGather(array{OrderInfoResult(2)}, array{OrderInfoResult(3)}, array{false},false)
                    VFXManager.BigObjetive_AudioPlayer.Play(Agent)
                    spawn:
                        Test_Delay_NewOrderDisplay(Player)
                else if(OrderInfoResult(1)?):
                    #Print("Refreshing order progress")
                    MainInventoryUI.RefreshOrdersQuests(OrderInfoResult(4))

    Test_Delay_NewOrderDisplay(Player : player)<suspends> : void =
        if(Agent := agent[Player]):
            MaybeBakery:=GetAgentBakery(Agent)
            if(Bakery:=MaybeBakery?, MainInventoryUI := Bakery.MaybeMainInventoryUI?, ActionUI := Bakery.MaybeActionInfoUI?):
                MainInventoryUI.Display_OrderCompleted()
                Sleep(5.0)
                if(Player.IsActive[]):
                    set MainInventoryUI.IsOrderCompleted_BeingDisplayed = false
                    MainInventoryUI.RefreshOrdersQuests(Bakery.PlayerDataCache)

    Test_AddUnlockHexaChallengeLV(Agent : agent) : void =
        MaybeBakery:=GetAgentBakery(Agent)
        if(Player:=player[Agent], ValidBakery:=MaybeBakery?, Player.IsActive[]<#,  PlayerData:=PlayersDataManager.GetPlayerData[Player]#>):
            var NewData : player_data_table =PlayersDataManager.AddDungeonAccumulated(ValidBakery.PlayerDataCache)
            ##Print("Trying to unlock new challenge")
            for(Index:=0..10):
                set NewData=PlayersDataManager.SetTotemHexaChallenge(NewData,Index)(0)
                set NewData=PlayersDataManager.AddAccumulatedHexaToken(NewData,MainFormula_DelayHexaGame(NewData.HexaGame_LVDelay))
                #set NewData=PlayersDataManager.AddResourcers(NewData,300.0,COAL,-1.0)(0)
                set NewData=Test_GivePets(Agent,NewData)
                #PlayersDataManager.SaveNewData(Player,PlayerData,NewData)
                if(MainInventoryUI := ValidBakery.MaybeMainInventoryUI?, ActionUI := ValidBakery.MaybeActionInfoUI?):
                    ValidBakery.SetNewPlayerDataCache(NewData)
                #PlayersDataManager.SaveAllNewData(Player,PlayerData,NewData, ValidBakery)
                #MainInventoryUI.DisplayCharacterMessage(0,"You have earned one challenge and dungeon token and 300 coal")

            #DisplayCharacterMessage(Agent,0,"Unlock new HexaChallenge: {NewLV}")
                


    Test_GivePets(Agent : agent, PlayerData : player_data_table) : player_data_table =
        var NewData : player_data_table = PlayerData
        if(Value:=IsGivenPets[Agent], Value?):
        else:
            if(set IsGivenPets[Agent] = true):
                #Druid
                DruidPet_0:=PlayersDataManager.CreatePet(pet_data_table{}, DRUID, 299, 20, 20)
                set NewData=PlayersDataManager.AddNewPet(NewData,DruidPet_0)
                DruidPet_1:=PlayersDataManager.CreatePet(pet_data_table{}, DRUID, 299, 20, 20)
                set NewData=PlayersDataManager.AddNewPet(NewData,DruidPet_1)
                DruidPet_2:=PlayersDataManager.CreatePet(pet_data_table{}, DRUID, 299, 20, 20)
                set NewData=PlayersDataManager.AddNewPet(NewData,DruidPet_2)
                DruidPet_3:=PlayersDataManager.CreatePet(pet_data_table{}, DRUID, 299, 20, 20)
                set NewData=PlayersDataManager.AddNewPet(NewData,DruidPet_3)
                DruidPet_4:=PlayersDataManager.CreatePet(pet_data_table{}, DRUID, 299, 20, 20)
                set NewData=PlayersDataManager.AddNewPet(NewData,DruidPet_4)
                #Tank
                TankPet_0:=PlayersDataManager.CreatePet(pet_data_table{}, TANK, 299, 20, 20)
                set NewData=PlayersDataManager.AddNewPet(NewData,TankPet_0)
                TankPet_1:=PlayersDataManager.CreatePet(pet_data_table{}, TANK, 299, 20, 20)
                set NewData=PlayersDataManager.AddNewPet(NewData,TankPet_1)
                TankPet_2:=PlayersDataManager.CreatePet(pet_data_table{}, TANK, 299, 20, 20)
                set NewData=PlayersDataManager.AddNewPet(NewData,TankPet_2)
                TankPet_3:=PlayersDataManager.CreatePet(pet_data_table{}, TANK, 299, 20, 20)
                set NewData=PlayersDataManager.AddNewPet(NewData,TankPet_3)
                TankPet_4:=PlayersDataManager.CreatePet(pet_data_table{}, TANK, 299, 20, 20)
                set NewData=PlayersDataManager.AddNewPet(NewData,TankPet_4)
                #Fighter
                FighterPet_0:=PlayersDataManager.CreatePet(pet_data_table{}, FIGHTER, 299, 20, 20)
                set NewData=PlayersDataManager.AddNewPet(NewData,FighterPet_0)
                FighterPet_1:=PlayersDataManager.CreatePet(pet_data_table{}, FIGHTER, 299, 20, 20)
                set NewData=PlayersDataManager.AddNewPet(NewData,FighterPet_1)
                FighterPet_2:=PlayersDataManager.CreatePet(pet_data_table{}, FIGHTER, 299, 20, 20)
                set NewData=PlayersDataManager.AddNewPet(NewData,FighterPet_2)
                FighterPet_3:=PlayersDataManager.CreatePet(pet_data_table{}, FIGHTER, 299, 20, 20)
                set NewData=PlayersDataManager.AddNewPet(NewData,FighterPet_3)
                FighterPet_4:=PlayersDataManager.CreatePet(pet_data_table{}, FIGHTER, 299, 20, 20)
                set NewData=PlayersDataManager.AddNewPet(NewData,FighterPet_4)
                #Hunter
                HunterPet_0:=PlayersDataManager.CreatePet(pet_data_table{}, HUNTER, 299, 20, 20)
                set NewData=PlayersDataManager.AddNewPet(NewData,HunterPet_0)
                HunterPet_1:=PlayersDataManager.CreatePet(pet_data_table{}, HUNTER, 299, 20, 20)
                set NewData=PlayersDataManager.AddNewPet(NewData,HunterPet_1)
                HunterPet_2:=PlayersDataManager.CreatePet(pet_data_table{}, HUNTER, 299, 20, 20)
                set NewData=PlayersDataManager.AddNewPet(NewData,HunterPet_2)
                HunterPet_3:=PlayersDataManager.CreatePet(pet_data_table{}, HUNTER, 299, 20, 20)
                set NewData=PlayersDataManager.AddNewPet(NewData,HunterPet_3)
                HunterPet_4:=PlayersDataManager.CreatePet(pet_data_table{}, HUNTER, 299, 20, 20)
                set NewData=PlayersDataManager.AddNewPet(NewData,HunterPet_4)
                #ADC
                ADCPet_0:=PlayersDataManager.CreatePet(pet_data_table{}, ADC, 299, 20, 20)
                set NewData=PlayersDataManager.AddNewPet(NewData,ADCPet_0)
                ADCPet_1:=PlayersDataManager.CreatePet(pet_data_table{}, ADC, 299, 20, 20)
                set NewData=PlayersDataManager.AddNewPet(NewData,ADCPet_1)
                ADCPet_2:=PlayersDataManager.CreatePet(pet_data_table{}, ADC, 299, 20, 20)
                set NewData=PlayersDataManager.AddNewPet(NewData,ADCPet_2)
                ADCPet_3:=PlayersDataManager.CreatePet(pet_data_table{}, ADC, 299, 20, 20)
                set NewData=PlayersDataManager.AddNewPet(NewData,ADCPet_3)
                ADCPet_4:=PlayersDataManager.CreatePet(pet_data_table{}, ADC, 299, 20, 20)
                set NewData=PlayersDataManager.AddNewPet(NewData,ADCPet_4)
        return NewData

    Test_CompleteQuest(Agent : agent) : void =
        if(Player := player[Agent], Player.IsActive[]):
            MaybeBakery:=GetAgentBakery(Agent)
            if(Bakery:=MaybeBakery?, MainInventoryUI := Bakery.MaybeMainInventoryUI?, ActionUI := Bakery.MaybeActionInfoUI?):
                #DisplayCharacterMessage(Agent,0,"Tutorial completed! Now complete the next ")
                QuestInfoResult:=QuestsAndOrdersDatabase.AddAgentQuest_ProgressV2(Player,Bakery.PlayerDataCache,10000000.0,Bakery.PlayerDataCache.QuestAssigned_ObjetiveType)
                Bakery.SetNewPlayerDataCache(QuestInfoResult(4))

    Test_CompleteAllTutorials(Agent : agent) : void =
        if(Player := player[Agent], Player.IsActive[]):
            MaybeBakery:=GetAgentBakery(Agent)
            if(Bakery:=MaybeBakery?):
                NewData:=PlayersDataManager.SetNumQuestsCompleted(Bakery.PlayerDataCache,9999)
                Bakery.SetNewPlayerDataCache(NewData)

    TryCompleteQuest_Paid(Agent : agent) : void =
        if(Player := player[Agent], Player.IsActive[]):
            MaybeBakery:=GetAgentBakery(Agent)
            if(Bakery:=MaybeBakery?):
                if(Bakery.PlayerDataCache.QuestAssigned_NumQuest <> -99):
                    if(Bakery.PlayerDataCache.NumQuestsCompleted > Quest0_Index):
                        if(Bakery.PlayerDataCache.Inventory_Coal >= SkipQuestCoal_Cost*1.0):
                            DisplayCharacterMessage(Agent,G_CHARACTER,"Paid {SkipQuestCoal_Cost} to auto-complete your Quest", ?Sound:=SMALLOBJETIVE_SOUND)
                            QuestInfoResult:=QuestsAndOrdersDatabase.AddAgentQuest_ProgressV2(Player,Bakery.PlayerDataCache,10000000.0,Bakery.PlayerDataCache.QuestAssigned_ObjetiveType)
                            var NewData : player_data_table = QuestInfoResult(4)
                            set NewData = PlayersDataManager.WithdrawResource(NewData,SkipQuestCoal_Cost*1.0, COAL)
                            Bakery.SetNewPlayerDataCache(NewData)
                        else:
                            DisplayCharacterMessage(Agent,I_CHARACTER,"You need {GetOptimizedMagnitudeLetter((SkipQuestCoal_Cost*1.0) - Bakery.PlayerDataCache.Inventory_Coal)} more Coal", ?Sound:=ERROR_SOUND)
                    else:
                        if(Bakery.PlayerDataCache.Language=ENGLISH):
                            DisplayCharacterMessage(Agent,G_CHARACTER,"It is recommended not to skip the tutorial", ?Sound:=MESSAGE_SOUND)
                        else if(Bakery.PlayerDataCache.Language=SPANISH):
                            DisplayCharacterMessage(Agent,G_CHARACTER,"Se recomienda no saltar el tutorial", ?Sound:=MESSAGE_SOUND)
                        else if(Bakery.PlayerDataCache.Language=FRENCH):
                            DisplayCharacterMessage(Agent,G_CHARACTER,"Il est recommandé de ne pas sauter le tutoriel", ?Sound:=MESSAGE_SOUND)
                        else if(Bakery.PlayerDataCache.Language=PORTUGUESE):
                            DisplayCharacterMessage(Agent,G_CHARACTER,"É recomendado não pular o tutorial", ?Sound:=MESSAGE_SOUND)
                        QuestInfoResult:=QuestsAndOrdersDatabase.AddAgentQuest_ProgressV2(Player,Bakery.PlayerDataCache,10000000.0,Bakery.PlayerDataCache.QuestAssigned_ObjetiveType)
                        var NewData : player_data_table = QuestInfoResult(4)
                        #set NewData = PlayersDataManager.WithdrawResource(NewData,SkipQuestCoal_Cost*1.0, COAL)
                        Bakery.SetNewPlayerDataCache(NewData)
                    
                else:
                    DisplayCharacterMessage(Agent,I_CHARACTER,"You do not have a Quest assigned", ?Sound:=ERROR_SOUND)
                

    #Super energy / coal
    AskForHelp(Agent : agent) : void =
        AskForHelpMessage.SetText(GetAgentName(Agent))
        AskForHelpMessage.Show()

    IsThereSharedSuperEnergy(CallFromIndex : int) : logic =
        var IsTheOtherSharedEnergy : logic = false
        for(Index:=0..AllBakers.Length-1, MaybeBakery := AllBakers[Index]):
            if(Bakery := MaybeBakery?, AgentOwner := Bakery.MaybeAgentOwner?, PlayerOwner:=player[AgentOwner], PlayerOwner.IsActive[], Bakery.PlayerDataCache.SuperEnergy_Owned > 0.0, CallFromIndex<>Index):
                set IsTheOtherSharedEnergy = true
        return IsTheOtherSharedEnergy

    Test_GrantAgedSapphires(Agent : agent) : void =
        MaybeBakery:=GetAgentBakery(Agent)
        if:
            Player := player[Agent]
            Player.IsActive[]
            ValidBakery:=MaybeBakery?
            #PlayerData:=PlayersDataManager.GetPlayerData[Player]
        then:
            FinalAmount_AgedSapphires := ValidBakery.PlayerDataCache.Inventory_AgedSapphires + 10.0
            NewData := PlayersDataManager.AddResourcers(ValidBakery.PlayerDataCache, FinalAmount_AgedSapphires, AGEDSAPPHIRES, -1.0)(0)
            
            ValidBakery.SetNewPlayerDataCache(NewData)
            ValidBakery.RefreshInventoryUI(NewData)
        else:
            PrintWarning("Error on granting sapphires")

    Test_GrantMutantMaterial(Agent : agent) : void =
        MaybeBakery:=GetAgentBakery(Agent)
        if:
            Player := player[Agent]
            Player.IsActive[]
            ValidBakery:=MaybeBakery?
            #PlayerData:=PlayersDataManager.GetPlayerData[Player]
        then:
            FinalAmount_MutantMaterial := (ValidBakery.PlayerDataCache.Inventory_MutantMaterial*10) + 5.0
            NewData := PlayersDataManager.AddResourcers(ValidBakery.PlayerDataCache, FinalAmount_MutantMaterial, MUTANTMATERIAL, -1.0)(0)
            
            ValidBakery.SetNewPlayerDataCache(NewData)
            ValidBakery.RefreshInventoryUI(NewData)
        else:
            PrintWarning("Error on granting sapphires")

    Test_GrantCoal(Agent : agent) : void =
        MaybeBakery:=GetAgentBakery(Agent)
        if:
            Player := player[Agent]
            Player.IsActive[]
            ValidBakery:=MaybeBakery?
            #PlayerData:=PlayersDataManager.GetPlayerData[Player]
        then:
            FinalAmount_Coal := ValidBakery.PlayerDataCache.Inventory_Coal + 100.0
            NewData := PlayersDataManager.AddResourcers(ValidBakery.PlayerDataCache, FinalAmount_Coal, COAL, -1.0)(0)
            
            ValidBakery.SetNewPlayerDataCache(NewData)
            ValidBakery.RefreshInventoryUI(NewData)
        else:
            PrintWarning("Error on granting sapphires")

    Test_GrantResourcers(Agent : agent) : void =
        MaybeBakery:=GetAgentBakery(Agent)
        if:
            Player := player[Agent]
            Player.IsActive[]
            ValidBakery:=MaybeBakery?
            #PlayerData:=PlayersDataManager.GetPlayerData[Player]
        then:
            AmountWater := (ValidBakery.PlayerDataCache.Inventory_Water*10.0) + 10.0
            AmountFlour := (ValidBakery.PlayerDataCache.Inventory_Flour*10.0) + 10.0
            AmountCheese := (ValidBakery.PlayerDataCache.Inventory_Cheese*10.0) + 10.0
            AmountPimentos := (ValidBakery.PlayerDataCache.Inventory_Pimentos*10.0) + 10.0
            
            ValidBakery.SaveResourcers(Agent,array{WATER,FLOUR,CHEESE,PIMENTOS},array{AmountWater,AmountFlour,AmountCheese,AmountPimentos})
            ValidBakery.RefreshInventoryUI(ValidBakery.PlayerDataCache)
        else:
            PrintWarning("Error on granting sapphires")

    Test_AddTutorialPhase(Agent : agent) : void =
        MaybeBakery:=GetAgentBakery(Agent)
        if:
            Player := player[Agent]
            Player.IsActive[]
            ValidBakery:=MaybeBakery?
            UKTutorial:=ValidBakery.MaybeUKTutorial?
        then:
            set UKTutorial.CurrentPhase = UKTutorial.CurrentPhase + 1
            UKTutorial.TeleportToCurrentPhase(Agent)
            if(UKTutorial.CurrentPhase=0):
                UKTutorial.PlayPhase0()
            else if(UKTutorial.CurrentPhase=1):
                UKTutorial.PlayPhase1()
            else if(UKTutorial.CurrentPhase=2):
                UKTutorial.PlayPhase2()
            else if(UKTutorial.CurrentPhase=3):
                UKTutorial.PlayPhase3()
            else if(UKTutorial.CurrentPhase=4):
                UKTutorial.PlayPhase4()
                
    Test_UpgradeBackpack(Agent : agent) : void =
        MaybeBakery:=GetAgentBakery(Agent)
        if:
            Player := player[Agent]
            Player.IsActive[]
            ValidBakery:=MaybeBakery?
            #PlayerData:=PlayersDataManager.GetPlayerData[Player]
        then:
            
            PlayerDataCache := ValidBakery.PlayerDataCache

            var AllBackPacks : []int = PlayerDataCache.AllLV_BackpackLimits

            if(LastLevel := AllBackPacks[WATER], set AllBackPacks[WATER] = LastLevel + 1):
            if(LastLevel := AllBackPacks[FLOUR], set AllBackPacks[FLOUR] = LastLevel + 1):
            if(LastLevel := AllBackPacks[CHEESE], set AllBackPacks[CHEESE] = LastLevel + 1):
            if(LastLevel := AllBackPacks[PIMENTOS], set AllBackPacks[PIMENTOS] = LastLevel + 1):
            #SAVE NEW DATA
            ValidBakery.SetNewPlayerDataCache(PlayersDataManager.SetAllLV_BackpackLimits(PlayerDataCache,AllBackPacks))
            #PlayersDataManager.SaveAllNewData(PlayerOwner, PlayerData, NewTableCompleted, Self)
            ValidBakery.GenerateAllBackPackLimits(PlayerDataCache)
            ValidBakery.RefreshInventoryUI(PlayerDataCache)
            ValidBakery.CheckQuestProgress(1.0,IMPROVEBACKPACK_TUTORIAL_TYPE)
            ValidBakery.GameManager.VFXManager.AwardAccolade(INVENTORY_ACCOLADE,Agent)
        else:
            PrintWarning("Error on Test_UpgradeBackpack")

    Test_LoadDataTable(Agent : agent) : void =
        SavePlayerDataTables()
        if:
            Player := player[Agent]
            Player.IsActive[]
            PlayerData:=PlayersDataManager.LoadPlayerCache[Player]
            #PlayerData2:=PlayersDataManager2.LoadPlayerCache[Player]
        then:    
            #NewPlayerData2:=PlayersDataManager2.SetSapphires(PlayerData.Inventory_Sapphires)
            DisplayCharacterMessage(Agent, W_CHARACTER, "Saved Sapphires on Data Table was: {GetOptimizedMagnitudeLetter(PlayerData.Inventory_Sapphires)}")


    Test_GrantSapphires(Agent : agent) : void =
        MaybeBakery:=GetAgentBakery(Agent)
        if:
            Player := player[Agent]
            Player.IsActive[]
            ValidBakery:=MaybeBakery?
            #PlayerData:=PlayersDataManager.GetPlayerData[Player]
        then:
            FinalAmount_Sapphires := (ValidBakery.PlayerDataCache.Inventory_Sapphires*10.0) + 10.0
            NewData := PlayersDataManager.AddResourcers(ValidBakery.PlayerDataCache, FinalAmount_Sapphires, SAPPHIRES, -1.0)(0)
            
            ValidBakery.SetNewPlayerDataCache(NewData)
            ValidBakery.RefreshInventoryUI(NewData)
        else:
            PrintWarning("Error on granting sapphires")
                    
    DisplayCoalUI(Agent : agent) : void =
        MaybeBakery:=GetAgentBakery(Agent)
        if(ValidBakery := MaybeBakery?):
            ValidBakery.Display_DropCoal(Agent)

    DisplayImproveDungeonTokens(Agent : agent) : void =
        MaybeBakery:=GetAgentBakery(Agent)
        if(ValidBakery := MaybeBakery?):
            ValidBakery.Display_ImproveDungeonTokens(Agent)

    DisplayImproveHexaGameTokens(Agent : agent) : void =
        MaybeBakery:=GetAgentBakery(Agent)
        if(ValidBakery := MaybeBakery?):
            ValidBakery.Display_ImproveHexaGameTokens(Agent)

    DisplayImproveHammer(Agent : agent) : void =
        MaybeBakery:=GetAgentBakery(Agent)
        if(ValidBakery := MaybeBakery?):
            ValidBakery.Display_ImproveWeapons(SWORD_TYPE)

    DisplayImproveBow(Agent : agent) : void =
        MaybeBakery:=GetAgentBakery(Agent)
        if(ValidBakery := MaybeBakery?):
            ValidBakery.Display_ImproveWeapons(BOW_TYPE)

    IsSharedCoalOverFlow() : logic =
        var IsOverFlow : logic = false
        if(SuperEnergy_Shared>MaxTime_SuperEnergy):
            set IsOverFlow = true
        return IsOverFlow

    UpdateSharedCoal() : logic =
        var IsOverFlow : logic = false
        
        CalculeSuperEnergy()

        Timer_InsertCoal.Reset()
        Timer_InsertCoal.SetActiveDuration(SuperEnergy_Shared) 
        Timer_InsertCoal.SetMaxDuration(SuperEnergy_Shared) 
        Timer_InsertCoal.Reset()
        Timer_InsertCoal.SetActiveDuration(SuperEnergy_Shared) 
        Timer_InsertCoal.SetMaxDuration(SuperEnergy_Shared) 
        VFXManager.CoalGenerator_Audio.Play()
        if(ValidProp:=MaybeCoalGeneratorProp_SuperEnergy?):
        else:
            SpawnTuple:=SpawnProp(VFXManager.CoalGeneratorProp_SuperEnergy,CoalGeneratorProp.GetTransform())
            set MaybeCoalGeneratorProp_SuperEnergy = SpawnTuple(0)
        if(SuperEnergy_Shared>MaxTime_SuperEnergy):
            set IsOverFlow = true

        return IsOverFlow

    ResetPlayer<private>(Agent : agent) : void =
        if(DebugOn?):
            if:
                Player := player[Agent]
                set AllPlayersDataMap[Player] = player_data_table{}
            then:
                GeneralMobsManager.WeaponsRemover.Remove(Agent)
                OnPlayerRemoved(Player)
                OnPlayerAdded(Player)
            else:
                DisplayCharacterMessage(Agent, W_CHARACTER, "DEBUG OFF")

    TryGetMastery(Agent : agent) : void =
        if(AgentBakery:=GetAgentBakery_Transacted(Agent)?):
            AgentBakery.Display_MasteryUI(Agent)

            <#
    RunBasicChecks() : void =
        if(AllBarrierArea.Length = AllButtonsArea.Length = AllBillboardsArea.Length):
        else:
            #Print("SUPER FATAL ERROR!\n------------\nAllZoneArea.Length, AllButtonsArea.Length and AllBillboardsArea.Length ARE NOT EQUAL!")
            #>

    InitHardbread() : void =
        TaggedActors := FindCreativeObjectsWithTag(hardbread{})
        for(Actor : TaggedActors):
            if(HardBreadManager:=hardbread_managerv2[Actor]):
                HardBreadManager.SetGameManager(Self)
                set AllHardBreads = AllHardBreads + array{HardBreadManager}
            if(HardBreadMob:=npc_spawner_device[Actor]):
                set AllMobSpanwers = AllMobSpanwers + array{HardBreadMob}
        var InestableIndex : int = 0
        for(Bread : AllHardBreads):
            Bread.SetGameManager(Self)
            if(Bread.IsInestable?):
                Bread.SetInestableHBIndex(InestableIndex)
                MiniGamesManager.AddInestableHB(Bread)
                set InestableIndex = InestableIndex + 1
        for(Points_Props : AllWeakPoints_Props):
            set IsUsedWeakPoint = IsUsedWeakPoint + array{false}
            set WeakPoint_OrTransform = WeakPoint_OrTransform + array{Points_Props.GetTransform()}

        set AssignedHardBreads=
            for(Index:=0..AllWeakPoints_Props.Length-1):
                false
        
                
    SetCostAreas() : void =
        #InitCosts
        for(NumArea:=0..AllButtonsArea.Length-1):
            if:
                Button := AllButtonsArea[NumArea]
                Barrier := AllBarrierArea[NumArea]
                Billboard := AllBillBoardsArea[NumArea]
                #NameString := AllAreaNames[Index]
            then:
                if(AreaCost:=PickaxeCalculator.AllLVs_ZoneCosts[NumArea]):
                    if(NumArea=HB_MAXLEVEL-1):
                        AreaCostFormated := GetOptimizedMagnitudeLetter(AreaCost)
                        CharFormated:=GetCharMagnitudeLetter(AreaCost)
                        SetBillBoardCost(Billboard,CharFormated,?IsAgedSapphires := true)
                        Button.SetInteractionText(BasicMessage("Unlock next area x{AreaCostFormated} Aged Sapphires | +5% Efficiency"))
                    else:
                        AreaCostFormated := GetOptimizedMagnitudeLetter(AreaCost)
                        CharFormated:=GetCharMagnitudeLetter(AreaCost)
                        SetBillBoardCost(Billboard,CharFormated)
                        Button.SetInteractionText(BasicMessage("Unlock next area x{AreaCostFormated} Sapphires | +5% Efficiency"))

        #Check players that have unlocked the area
        for(Player : GetPlayspace().GetPlayers(), Agent := agent[Player], PlayerData := PlayersDataManager.LoadPlayerCache[Player]):
            CheckAreasToUnlock(PlayerData,Agent)
            
        GeneralMobsManager.InitHBHealthArray()# After setting cost, we can set the prev damage

    SetBillBoardCost(BillBoardProp : creative_prop, AmountString : []char, ?IsAgedSapphires : logic = false) : void =
        BillBoardProp.SetMaterial(Materials.MI_0,?Index:=1)
        BillBoardProp.SetMaterial(Materials.BP_Nothing,?Index:=2)
        BillBoardProp.SetMaterial(Materials.MI_0,?Index:=3)
        BillBoardProp.SetMaterial(Materials.BP_Nothing,?Index:=4)
        BillBoardProp.SetMaterial(Materials.MI_0,?Index:=5)
        BillBoardProp.SetMaterial(Materials.BP_Nothing,?Index:=6)
        if(IsAgedSapphires?):
            BillBoardProp.SetMaterial(Images.AgedSapphire_Icon_Mat_Inst,?Index:=7)
        else:
            BillBoardProp.SetMaterial(Images.Sapphire_Icon_Mat,?Index:=7)
        ##Print("##########################")
        var DigitIndex : int = 1
        var PointIndex : int = 2
        #Log("#############",LOG_LEVEL_DEBUG5)
        for(Index:=0..3, Char:=AmountString[Index]):
            #Log("Character: {Char}",LOG_LEVEL_DEBUG5)
            #Numbers
            if(Char='.'):
                BillBoardProp.SetMaterial(Materials.MI_Dot,?Index:=(PointIndex))
            if(Char='0', MaterialToUse := AllMaterialnumbers[0]):
                BillBoardProp.SetMaterial(MaterialToUse,?Index:=(DigitIndex))
                set DigitIndex = DigitIndex + 2
                if(Index>0):
                    set PointIndex = PointIndex + 2
            if(Char='1', MaterialToUse := AllMaterialnumbers[1]):
                BillBoardProp.SetMaterial(MaterialToUse,?Index:=(DigitIndex))
                set DigitIndex = DigitIndex + 2
                if(Index>0):
                    set PointIndex = PointIndex + 2
            if(Char='2', MaterialToUse := AllMaterialnumbers[2]):
                BillBoardProp.SetMaterial(MaterialToUse,?Index:=(DigitIndex))
                set DigitIndex = DigitIndex + 2
                if(Index>0):
                    set PointIndex = PointIndex + 2
            if(Char='3', MaterialToUse := AllMaterialnumbers[3]):
                BillBoardProp.SetMaterial(MaterialToUse,?Index:=(DigitIndex))
                set DigitIndex = DigitIndex + 2
                if(Index>0):
                    set PointIndex = PointIndex + 2
            if(Char='4', MaterialToUse := AllMaterialnumbers[4]):
                BillBoardProp.SetMaterial(MaterialToUse,?Index:=(DigitIndex))
                set DigitIndex = DigitIndex + 2
                if(Index>0):
                    set PointIndex = PointIndex + 2
            if(Char='5', MaterialToUse := AllMaterialnumbers[5]):
                BillBoardProp.SetMaterial(MaterialToUse,?Index:=(DigitIndex))
                set DigitIndex = DigitIndex + 2
                if(Index>0):
                    set PointIndex = PointIndex + 2
            if(Char='6', MaterialToUse := AllMaterialnumbers[6]):
                BillBoardProp.SetMaterial(MaterialToUse,?Index:=(DigitIndex))
                set DigitIndex = DigitIndex + 2
                if(Index>0):
                    set PointIndex = PointIndex + 2
            if(Char='7', MaterialToUse := AllMaterialnumbers[7]):
                BillBoardProp.SetMaterial(MaterialToUse,?Index:=(DigitIndex))
                set DigitIndex = DigitIndex + 2
                if(Index>0):
                    set PointIndex = PointIndex + 2
            if(Char='8', MaterialToUse := AllMaterialnumbers[8]):
                BillBoardProp.SetMaterial(MaterialToUse,?Index:=(DigitIndex))
                set DigitIndex = DigitIndex + 2
                if(Index>0):
                    set PointIndex = PointIndex + 2
            if(Char='9', MaterialToUse := AllMaterialnumbers[9]):
                BillBoardProp.SetMaterial(MaterialToUse,?Index:=(DigitIndex))
                set DigitIndex = DigitIndex + 2
                if(Index>0):
                    set PointIndex = PointIndex + 2
        if(CharLetter:=AmountString[5]):
            if(CharLetter='k'):
                BillBoardProp.SetMaterial(Materials.MI_K,?Index:=6)
            else if(CharLetter='M'):
                BillBoardProp.SetMaterial(Materials.MI_M,?Index:=6)
            else if(CharLetter='G'):
                BillBoardProp.SetMaterial(Materials.MI_G,?Index:=6)
            else if(CharLetter='T'):
                BillBoardProp.SetMaterial(Materials.MI_T,?Index:=6)
            else if(CharLetter='P'):
                BillBoardProp.SetMaterial(Materials.MI_P,?Index:=6)
            else if(CharLetter='E'):
                BillBoardProp.SetMaterial(Materials.MI_E,?Index:=6)
            else if(CharLetter='Z'):
                BillBoardProp.SetMaterial(Materials.MI_Z,?Index:=6)
            else if(CharLetter='Y'):
                BillBoardProp.SetMaterial(Materials.MI_Y,?Index:=6)
            else if(CharLetter='R'):
                BillBoardProp.SetMaterial(Materials.MI_R,?Index:=6)
            else if(CharLetter='Q'):
                BillBoardProp.SetMaterial(Materials.MI_Q,?Index:=6)
        else:
            if(CharLetter:=AmountString[4]):
                if(CharLetter='k'):
                    BillBoardProp.SetMaterial(Materials.MI_K,?Index:=6)
                else if(CharLetter='M'):
                    BillBoardProp.SetMaterial(Materials.MI_M,?Index:=6)
                else if(CharLetter='G'):
                    BillBoardProp.SetMaterial(Materials.MI_G,?Index:=6)
                else if(CharLetter='T'):
                    BillBoardProp.SetMaterial(Materials.MI_T,?Index:=6)
                else if(CharLetter='P'):
                    BillBoardProp.SetMaterial(Materials.MI_P,?Index:=6)
                else if(CharLetter='E'):
                    BillBoardProp.SetMaterial(Materials.MI_E,?Index:=6)
                else if(CharLetter='Z'):
                    BillBoardProp.SetMaterial(Materials.MI_Z,?Index:=6)
                else if(CharLetter='Y'):
                    BillBoardProp.SetMaterial(Materials.MI_Y,?Index:=6)
                else if(CharLetter='R'):
                    BillBoardProp.SetMaterial(Materials.MI_R,?Index:=6)
                else if(CharLetter='Q'):
                    BillBoardProp.SetMaterial(Materials.MI_Q,?Index:=6)


    #TryUnlockAreas
    CheckAreasToUnlock(PlayerData : player_data_table, Agent : agent) : void =
        if(Player := player[Agent]):
            for(Index:=0..AllBarrierArea.Length-1, Barrier := AllBarrierArea[Index]):
                if(Index=HB_MAXLEVEL-1):
                    if(PlayerData.LastAreaUnlocked?):
                        Log("LastAreaUnlocked Unlocked: {PlayerData.LV_Area}",LOG_LEVEL_DEBUG7)
                        Barrier.AddToIgnoreList(Agent)
                    else:
                        Log("LastAreaUnlocked Locked: {PlayerData.LV_Area}",LOG_LEVEL_DEBUG7)
                        Barrier.RemoveFromIgnoreList(Agent)
                else:
                    if(Index > PlayerData.LV_Area):
                        Log("Removed from ignore list. Index, PlayerData.LV_Area is: {PlayerData.LV_Area}",LOG_LEVEL_MAP)
                        Barrier.RemoveFromIgnoreList(Agent)
                    else:
                        Log("Added to ignore list. Index, PlayerData.LV_Area is: {PlayerData.LV_Area}",LOG_LEVEL_MAP)
                        Barrier.AddToIgnoreList(Agent)

    ShowDamageToPlayer(Damage : float, WeaponType : int, Agent : agent, IsCritical : logic): void =
        ##Print("Trying to show damage to player on UI")
        if(WeaponType=BOW_TYPE):
            DisplayResourcersGather(Agent, array{Damage}, array{BOW_TYPE}, IsCritical)
        else if(WeaponType=SWORD_TYPE):
            DisplayResourcersGather(Agent, array{Damage}, array{SWORD_TYPE},IsCritical)
        else if(WeaponType=-1):
            DisplayResourcersGather(Agent, array{Damage}, array{SWORD_TYPE},IsCritical)


    TryUnlockArea(Agent : agent, Area : int) : void =
        Log("Try Unlock area",LOG_LEVEL_MAP)
        MaybeBakery:=GetAgentBakery(Agent)
        if:
            Player := player[Agent]
            #PlayerData := PlayersDataManager.GetPlayerData[Player]
            Button := AllButtonsArea[Area]
            Barrier := AllBarrierArea[Area]
            #Billboard := AllBillboardsArea[Area]
            #NameString := AllAreaNames[Area]
            Cost := PickaxeCalculator.AllLVs_ZoneCosts[Area]
            BakeryMachinesManager := MaybeBakery?
        then:
            if(Area=HB_MAXLEVEL-1):
                Log("Last Area: {Area}, PlayerDataCache.LV_Area: {BakeryMachinesManager.PlayerDataCache.LV_Area}, Cost: {Cost} Aged Sapphires",LOG_LEVEL_DEBUG7)
                if(BakeryMachinesManager.PlayerDataCache.LV_Area+1=HB_MAXLEVEL-1, not BakeryMachinesManager.PlayerDataCache.LastAreaUnlocked?):
                    if(BakeryMachinesManager.PlayerDataCache.Inventory_AgedSapphires>=Cost):
                        Log("Last Area unlocked!",LOG_LEVEL_DEBUG7) 
                        DisplayResourcersGather(Agent,array{-Cost}, array{AGEDSAPPHIRES}, false)
                        #VFXManager.PlayerPays_AudioPlayer.Play(Agent)
                        var NewPlayerData : player_data_table = PlayersDataManager.SetAgedSapphires(BakeryMachinesManager.PlayerDataCache,BakeryMachinesManager.PlayerDataCache.Inventory_AgedSapphires-Cost)
                        set NewPlayerData=PlayersDataManager.SetLV_Area(NewPlayerData,BakeryMachinesManager.PlayerDataCache.LV_Area+1)
                        set NewPlayerData=PlayersDataManager.SetLast_Area(NewPlayerData,true)
                        BakeryMachinesManager.SetNewPlayerDataCache(NewPlayerData)
                        QuestsAndOrdersDatabase.CheckLore(Player,BakeryMachinesManager.PlayerDataCache,?SpecialEvent:=SPECIALEVENT_MAINLORE)
                        #PlayersDataManager.SaveNewData(Player, PlayerData, DataWithNewLevel)
                        
                        #PlayersDataManager.SaveAllNewData(Player,PlayerData,DataWithNewLevel, BakeryMachinesManager)
                        CheckAreasToUnlock(NewPlayerData,Agent)
                        BakeryMachinesManager.RefreshInventoryUI(NewPlayerData)
                        VFXManager.Accolade_NewArea.Award(Agent)
                        VFXManager.BigObjetive_AudioPlayer.Play(Agent)
                        #VFXManager.PlayerPays_AudioPlayer.Play(Agent)
                        BakeryMachinesManager.CheckQuestProgress(1.0,UNLOCKAREA_TUTORIAL_TYPE)
                        Zone19Analytic.Submit(Agent)
                    else:
                        DisplayCharacterMessage(Agent,H_CHARACTER,"Hey buddy! You need {GetOptimizedMagnitudeLetter(Cost-BakeryMachinesManager.PlayerDataCache.Inventory_AgedSapphires)} more Aged Sapphires!")
                else if(BakeryMachinesManager.PlayerDataCache.LV_Area+1<HB_MAXLEVEL-1):
                    Display_UnlockPrevious(Agent)
                else if(BakeryMachinesManager.PlayerDataCache.LastAreaUnlocked?):
                    Display_AlreadyUnlocked(Agent)
            else:
                Log("Area: {Area}, PlayerData.LV_Area: {BakeryMachinesManager.PlayerDataCache.LV_Area}, Cost: {Cost}",LOG_LEVEL_DEBUG7)
                if(BakeryMachinesManager.PlayerDataCache.LV_Area=Area-1):
                    if(BakeryMachinesManager.PlayerDataCache.Inventory_Sapphires>=Cost):
                        Log("Area unlocked: {Area}",LOG_LEVEL_DEBUG7)
                        DisplayResourcersGather(Agent,array{-Cost}, array{SAPPHIRES}, false)
                        #VFXManager.PlayerPays_AudioPlayer.Play(Agent)
                        var NewPlayerData : player_data_table = PlayersDataManager.SetSapphires(BakeryMachinesManager.PlayerDataCache,BakeryMachinesManager.PlayerDataCache.Inventory_Sapphires-Cost)
                        set NewPlayerData=PlayersDataManager.SetLV_Area(NewPlayerData,BakeryMachinesManager.PlayerDataCache.LV_Area+1)
                        BakeryMachinesManager.SetNewPlayerDataCache(NewPlayerData)
                        QuestsAndOrdersDatabase.CheckLore(Player,BakeryMachinesManager.PlayerDataCache,?SpecialEvent:=SPECIALEVENT_MAINLORE)
                        #PlayersDataManager.SaveNewData(Player, PlayerData, DataWithNewLevel)
                        #PlayersDataManager.SaveAllNewData(Player,PlayerData,DataWithNewLevel, BakeryMachinesManager)
                        CheckAreasToUnlock(NewPlayerData,Agent)
                        BakeryMachinesManager.RefreshInventoryUI(NewPlayerData)
                        VFXManager.Accolade_NewArea.Award(Agent)
                        VFXManager.BigObjetive_AudioPlayer.Play(Agent)
                        #VFXManager.PlayerPays_AudioPlayer.Play(Agent)
                        BakeryMachinesManager.CheckQuestProgress(1.0,UNLOCKAREA_TUTORIAL_TYPE)
                        if(Area=15):
                            Zone15Analytic.Submit(Agent)
                        if(Area=16):
                            Zone16Analytic.Submit(Agent)
                        if(Area=17):
                            Zone17Analytic.Submit(Agent)
                        if(Area=18):
                            Zone18Analytic.Submit(Agent)
                    else:
                        Display_NotEnough(Agent,Cost-BakeryMachinesManager.PlayerDataCache.Inventory_Sapphires)   
                else if(BakeryMachinesManager.PlayerDataCache.LV_Area<Area-1):
                    Display_UnlockPrevious(Agent)
                else if(BakeryMachinesManager.PlayerDataCache.LV_Area>=Area):
                    Display_AlreadyUnlocked(Agent)
        else:
            PrintFatalError("On TryUnlockArea : Area: {Area} Invalid!")

    RunLoop()<suspends> : void =
        ##Print("Gaming loop runing")
        if(DebugOn?):
            for(Level:=0..30):
                HardBreadHealth:=GetOptimizedMagnitudeLetter(PickaxeCalculator.MainFormula_GenerateHardBreadHealth(Level,0))
                HardBreadPrev_Health:=GetOptimizedMagnitudeLetter(PickaxeCalculator.MainFormula_GenerateHardBreadHealth(Level,0)/PickaxeCalculator.MainFormula_GenerateHardBreadHealth(Level-1,0))
                Log("HardBread Lv{Level} with Health: {HardBreadHealth} has x{HardBreadPrev_Health} more health than the previous one",LOG_LEVEL_DEBUG3)
        var TimedHelp : int = 0
        <#
        if(LOG_LEVEL = LOG_LEVEL_DEBUG7):
            for(Purity:=0..300):
                Amount:=MainFormula_PetPurityCost(Purity)
                
                Log("For Purity {Purity} it will be required {GetOptimizedMagnitudeLetter(Amount)} Mutant Material",LOG_LEVEL_DEBUG7)#>
        loop:
            MakeBakersWork() #Delay 0.72
            Sleep(0.04) #0.80
            set TotalGameTime = TotalGameTime + 0.08
            ConsumeSuperEnergy()
            VFXManager.TrySteveBake(TotalGameTime)
            Sleep(0.04) #0.84
            set TotalGameTime = TotalGameTime + 0.04
            AwardAFKPlayers()
            Sleep(0.04) #0.92
            set TotalGameTime = TotalGameTime + 0.08
            GeneralMobsManager.RefreshSwordEffects()
            SavePlayerDataTables()
            Sleep(0.04) #0.96
            set TotalGameTime = TotalGameTime + 0.04
            MiniGamesManager.MiniGamesTick()
            Sleep(0.04) #1.00
            set TotalGameTime = TotalGameTime + 0.04
            GeneralMobsManager.CheckNPCSpawners()
            Sleep(0.04) #1.00
            MiniGamesManager.CheckScrollSpawn()
            if(TimedHelp>120):
                set TimedHelp = 0
                ExecuteTimedHelp()
            else:
                set TimedHelp = TimedHelp + 1
            GeneralMobsManager.CheckWeaponsEffects()
            Sleep(0.04) #1.00  

    ExecuteTimedHelp() : void =
        for(MaybeBakery : AllBakers, ValidBakery := MaybeBakery?, not ValidBakery.IsOnUKTutorial?, AgentOwner := ValidBakery.MaybeAgentOwner?, MainLoreState := ValidBakery.PlayerDataCache.AllLoreStates[0], MainLoreState<2, 3>ValidBakery.PlayerDataCache.LV_Area, ValidBakery.PlayerDataCache.Inventory_Fame<1000000.0):
            ValidBakery.SetNewPlayerDataCache(PlayersDataManager.AddResourcers(ValidBakery.PlayerDataCache,5.0,SAPPHIRES,0.0)(0))
            ValidBakery.DisplayResourcersGather(array{5.0}, array{SAPPHIRES}, false) 
            Position:=IsAgentOnArray(AgentOwner,QuestsAndOrdersDatabase.AllAgentsWatchingTutorial)
            if(Position=-1):
                DisplayCharacterMessage(AgentOwner, I_CHARACTER,"Take these Sapphires. Some help for you!",?Sound:=MESSAGE_SOUND, ?DisplayTime:=5.0)
            ValidBakery.RefreshInventoryUI(ValidBakery.PlayerDataCache)
          
    SavePlayerDataTables() : void =
        for(Player : GetPlayspace().GetPlayers(), Player.IsActive[], Agent := agent[Player]):
            if(AgentJoinTime:=AllAgentsJoinTime[Agent], AgentJoinTime+10.0<TotalGameTime):
                if(ValidBakery := GetAgentBakery_Transacted(Agent)?, OldData := PlayersDataManager.LoadPlayerCache[Player]):   
                    PlayersDataManager.SaveNewData(Player,OldData,ValidBakery.PlayerDataCache)
                    <#
        for(Index:=0..AllBakers.Length-1, MaybeBakery := AllBakers[Index]):
            if(Bakery := MaybeBakery?, AgentOwner := Bakery.MaybeAgentOwner?, PlayerOwner:=player[AgentOwner], PlayerOwner.IsActive[], OldData := PlayersDataManager.LoadPlayerCache[PlayerOwner]):
                PlayersDataManager.SaveNewData(PlayerOwner,OldData,Bakery.PlayerDataCache)#>
            #else:
            #    PrintFatalError("On SavePlayerDataTables() on Index: {Index}")

    AwardAFKPlayers() : void =
        for(Player : AllPlayersOnAFK_Zone):
            if(Player.IsActive[], Agent:=agent[Player]):
                if(Distance(Player.GetFortCharacter[].GetTransform().Translation,MutZone_AfkZone.GetTransform().Translation)<2000.0):
                    MaybeBakery:=GetAgentBakery(Agent)
                    if(ValidBakery:=MaybeBakery?):
                        ValidBakery.AwardAFKResourcers(Agent)
                        ValidBakery.CheckQuestProgress(1.0,USEAFKZONE_TUTORIAL_TYPE)
                else:
                    AgentExitsAFKZone(Agent) 

            else if(Agent:=agent[Player]):
                AgentExitsAFKZone(Agent)

    CalculeSuperEnergy() : void =
        #SharedEnergy
        set SuperEnergy_Shared = 0.0
        for(MaybeBakery : AllBakers, Bakery := MaybeBakery?, AgentOwner := Bakery.MaybeAgentOwner?, PlayerOwner:=player[AgentOwner], PlayerOwner.IsActive[]):
            set SuperEnergy_Shared = SuperEnergy_Shared + Bakery.PlayerDataCache.SuperEnergy_Owned
        if(SuperEnergy_Shared < 0.0):
            set SuperEnergy_Shared = 0.0 
        if(SuperEnergy_Shared<=0.0):
            Timer_InsertCoal.SetActiveDuration(0.0)
            if(CoalGeneratorProp_SuperEnergy:=MaybeCoalGeneratorProp_SuperEnergy?, CoalGeneratorProp_SuperEnergy.IsValid[]):
                CoalGeneratorProp_SuperEnergy.Dispose()
                VFXManager.CoalGenerator_Audio.Stop()
                set MaybeCoalGeneratorProp_SuperEnergy = false
        else:
            Timer_InsertCoal.SetMaxDuration(SuperEnergy_Shared)  
        

    ConsumeSuperEnergy() : void =
        #SharedEnergy
        set SuperEnergy_Shared = 0.0
        #OwnedEnergy
        for(MaybeBakery : AllBakers, Bakery := MaybeBakery?, AgentOwner := Bakery.MaybeAgentOwner?, PlayerOwner:=player[AgentOwner], PlayerOwner.IsActive[]):
            Bakery.TryConsumeSuperEnergy()
            set SuperEnergy_Shared = SuperEnergy_Shared + Bakery.PlayerDataCache.SuperEnergy_Owned
        if(SuperEnergy_Shared < 0.0):
            set SuperEnergy_Shared = 0.0 
        if(SuperEnergy_Shared<=0.0):
            Timer_InsertCoal.SetActiveDuration(0.0)
            if(CoalGeneratorProp_SuperEnergy:=MaybeCoalGeneratorProp_SuperEnergy?, CoalGeneratorProp_SuperEnergy.IsValid[]):
                CoalGeneratorProp_SuperEnergy.Dispose()
                VFXManager.CoalGenerator_Audio.Stop()
                set MaybeCoalGeneratorProp_SuperEnergy = false
        else:
            Timer_InsertCoal.SetMaxDuration(SuperEnergy_Shared)  

    GetAgentIndex(Agent : agent) : int =
        var Index : int = 0
            loop:
                if(MaybeBaker := AllBakers[Index], ValidBaker := MaybeBaker?, AgentOwner := ValidBaker.MaybeAgentOwner?, AgentOwner = Agent):
                    break
                else:
                    set Index = Index + 1
                if(Index > AllBakers.Length-1):
                    #Print("FATAL ERROR on GetAgentBakery. Bakery of Agent not found!")
                    if(Player := player[Agent], Character:=Agent.GetFortCharacter[], Player.IsActive[], Character.IsActive[]):
                        Character.Damage(99999.9)
                    break
        return Index

    SetEntityHealth(Agent : agent, Ammount : float, Limit : float, Type : int, InLevel : int) : void =
        if(Bakery := GetAgentBakery_Transacted(Agent)?, EntityHealthUI:=Bakery.MaybeEntityHealthUI?):
            EntityHealthUI.SetInfo(Ammount,Limit,Type,InLevel)

    GetAgentBakery_Transacted(Agent : agent)<transacts> : ?bakery_machines_manager =
        var MaybeBakerOutPut : ?bakery_machines_manager = false
        if(Bakery := AllAgentsBakery[Agent]):
            #Log("Bakery found!",LOG_LEVEL_UIs)
            return option{Bakery}
        else:
            var Index : int = 0
            loop:
                if(MaybeBaker := AllBakers[Index], ValidBaker := MaybeBaker?, AgentOwner := ValidBaker.MaybeAgentOwner?, AgentOwner = Agent):
                    set MaybeBakerOutPut = MaybeBaker
                    if(set AllAgentsBakery[Agent] = ValidBaker):
                        return MaybeBaker
                    break
                else:
                    set Index = Index + 1
                if(Index > AllBakers.Length-1):
                    #Print("FATAL ERROR on GetAgentBakery. Bakery of Agent not found!")
                    break
                #Log("Bakery NOT found!",LOG_LEVEL_UIs)
            return MaybeBakerOutPut

    HandlerDamageCharacter(Character : fort_character) : void =
        Character.Damage(99999.9)

    GetAgentBakery(Agent : agent) : ?bakery_machines_manager =
        var MaybeBakerOutPut : ?bakery_machines_manager = false
        if(Bakery := AllAgentsBakery[Agent]):
            #Log("Bakery found!",LOG_LEVEL_UIs)
            return option{Bakery}
        else:
            var Index : int = 0
            loop:
                if(MaybeBaker := AllBakers[Index], ValidBaker := MaybeBaker?, AgentOwner := ValidBaker.MaybeAgentOwner?, AgentOwner = Agent):
                    set MaybeBakerOutPut = MaybeBaker
                    if(set AllAgentsBakery[Agent] = ValidBaker):
                        return MaybeBaker
                    break
                else:
                    set Index = Index + 1
                if(Index > AllBakers.Length-1):
                    #Print("FATAL ERROR on GetAgentBakery. Bakery of Agent not found!")
                    if(Player := player[Agent], Character:=Agent.GetFortCharacter[], Player.IsActive[], Character.IsActive[]):
                        Character.Damage(99999.9)
                    break
                Log("Bakery NOT found!",LOG_LEVEL_UIs)
            return MaybeBakerOutPut

    GetAgentHexaManager(Agent : agent) : ?hexa_minigame_manager =
        var MaybeHexaGameManager : ?hexa_minigame_manager = false
        var Index : int = 0
        loop:
            if(HexaManager:=AllHexaManager[Index], OwnerAgent := HexaManager.MaybeAgent?, OwnerAgent = Agent):
                set MaybeHexaGameManager = option{HexaManager}
                return option{HexaManager}
            set Index = Index + 1
            if(Index>AllHexaManager.Length-1):
                break
        return MaybeHexaGameManager

    CheckOrders_And_Quests(): void =
        for(Player : GetPlayspace().GetPlayers(), Player.IsActive[], Agent := agent[Player]):
            MaybeBakery:=GetAgentBakery(Agent)
            if(Bakery:=MaybeBakery?,MainInventoryUI := Bakery.MaybeMainInventoryUI?):
                var NewPlayerData : player_data_table = Bakery.PlayerDataCache
                #Hexagame
                if(NewPlayerData.HexaGame_DelayTime<=0.0, NewPlayerData.HexaGame_Accumulated < NewPlayerData.HexaGame_LVAccumulated+HexaGame_MinToken_Accumulated_MinusOne):
                    set NewPlayerData = PlayersDataManager.AddAccumulatedHexaToken(NewPlayerData,MainFormula_DelayHexaGame(NewPlayerData.HexaGame_LVDelay))
                    VFXManager.SmallObjetive_AudioPlayer.Play(Agent)
                    Bakery.DisplayResourcersGather(array{1.0},array{HEXATOKEN},false)
                    MainInventoryUI.RefreshAll(NewPlayerData, array{-1.0}, array{false})
                #Dungeon
                if(NewPlayerData.Dungeon_DelayTime<=0.0, NewPlayerData.Dungeon_Accumulated < NewPlayerData.Dungeon_LVAccumulated):
                    set NewPlayerData = PlayersDataManager.AddAccumulatedDungeonToken(NewPlayerData,MainFormula_DelayDungeon(NewPlayerData.Dungeon_LVDelay))
                    VFXManager.SmallObjetive_AudioPlayer.Play(Agent)
                    Bakery.DisplayResourcersGather(array{1.0},array{DUNGEONTOKEN},false)
                    MainInventoryUI.RefreshAll(NewPlayerData, array{-1.0}, array{false})
                #Order
                if(NewPlayerData.Order_DelayTime<=0.0, NewPlayerData.NumQuestsCompleted > TutorialOrder_Index, NewPlayerData.OrderAssigned_ProductRequested = -99):
                    TupleOrder:=QuestsAndOrdersDatabase.GetOrder(Bakery)
                    set NewPlayerData = PlayersDataManager.CreateNewOrder(NewPlayerData,TupleOrder(0),TupleOrder(1),TupleOrder(2),TupleOrder(3))
                    VFXManager.SmallObjetive_AudioPlayer.Play(Agent)
                    MainInventoryUI.RefreshOrders(NewPlayerData)
                #Quest
                if(NewPlayerData.Quest_DelayTime<=0.0, NewPlayerData.NumQuestsCompleted > Quest0_Index, NewPlayerData.QuestAssigned_NumQuest = -99):
                    QuestInfo:=QuestsAndOrdersDatabase.GetNewQuest(NewPlayerData.NumQuestsCompleted,NewPlayerData)
                    Log("Granting new quest: Quest: {QuestInfo(9)}, QuestObjetive: {QuestInfo(3)}",LOG_LEVEL_DEBUG6)
                    set NewPlayerData = PlayersDataManager.CreateNewQuest(NewPlayerData,QuestInfo(9),QuestInfo(3),QuestInfo(5),QuestInfo(6))
                    AssignQuest_MapTracker(Agent,QuestInfo(9),NewPlayerData.QuestAssigned_ObjetiveType)
                    VFXManager.SmallObjetive_AudioPlayer.Play(Agent)
                    MainInventoryUI.RefreshQuests(NewPlayerData)

                set NewPlayerData = PlayersDataManager.DecrementAllDelays(NewPlayerData,1.0)
                #PlayersDataManager.SaveNewData(Player,PlayerData,NewPlayerData)
                Bakery.SetNewPlayerDataCache(NewPlayerData)
                #PlayersDataManager.SaveAllNewData(Player,PlayerData,NewPlayerData, Bakery)
                MainInventoryUI.RefreshTimers(NewPlayerData)

    MakeBakersWork()<suspends> : void =
        #PrintWarning("Forced bakery 2 on MakeBakersWork")
        #IsDownButNotOut checker
        for(MaybeBakery : AllBakers, Bakery := MaybeBakery?, Bakery.IsBakerReady? ,AgentOwner := Bakery.MaybeAgentOwner?, PlayerOwner:=player[AgentOwner], PlayerOwner.IsActive[]):
            Bakery.CheckQuestTracker()
        #Sleep(0.04)
        set TotalGameTime = TotalGameTime + 0.08
        for(Player : GetPlayspace().GetPlayers(), Character := Player.GetFortCharacter[], Character.IsDownButNotOut[]):
            Character.Damage(9999.9)
        Sleep(0.04)
        set TotalGameTime = TotalGameTime + 0.08
        #Move Pet
        for(MaybeBakery : AllBakers, Bakery := MaybeBakery?, Bakery.IsBakerReady? ,AgentOwner := Bakery.MaybeAgentOwner?, PlayerOwner:=player[AgentOwner], PlayerOwner.IsActive[]):
            Bakery.PetsManager.MovePet()
        Sleep(0.04)
        set TotalGameTime = TotalGameTime + 0.08
        #Pet Action
        for(MaybeBakery : AllBakers, Bakery := MaybeBakery?, Bakery.IsBakerReady? ,AgentOwner := Bakery.MaybeAgentOwner?, PlayerOwner:=player[AgentOwner], PlayerOwner.IsActive[]):
            spawn:
                Bakery.PetsManager.CheckPetAction(TotalGameTime,true)
        Sleep(0.04)
        set TotalGameTime = TotalGameTime + 0.08
        #Orders and Quests
        CheckOrders_And_Quests()
        Sleep(0.04)
        #Take resourcers
        set TotalGameTime = TotalGameTime + 0.08
        for(MaybeBakery : AllBakers, Bakery := MaybeBakery?, Bakery.IsBakerReady? ,AgentOwner := Bakery.MaybeAgentOwner?, PlayerOwner:=player[AgentOwner], PlayerOwner.IsActive[]):
            Bakery.FillPasiveMainDeposits()
            Bakery.TakeWater()
            Bakery.TakeFlour()
            Bakery.TakeCheese()
            Bakery.TakePimentos()
            Bakery.UpgradeResourcersContainersUI()
            #set NewPlayerData = Bakery.UpgradeValveButtonsDescriptions(NewPlayerData)
        Sleep(0.04)
        set TotalGameTime = TotalGameTime + 0.08
        #Update main deposits
        for(MaybeBakery : AllBakers, Bakery := MaybeBakery?, Bakery.IsBakerReady? ,AgentOwner := Bakery.MaybeAgentOwner?, PlayerOwner:=player[AgentOwner], PlayerOwner.IsActive[]):
            Bakery.UpdateAllMainDeposits()
            #Bakery.UpdateSubdeposits()
        Sleep(0.04)
        set TotalGameTime = TotalGameTime + 0.08
        #Update secondary deposits
        for(MaybeBakery : AllBakers, Bakery := MaybeBakery?,  Bakery.IsBakerReady? ,AgentOwner := Bakery.MaybeAgentOwner?, PlayerOwner:=player[AgentOwner], PlayerOwner.IsActive[]):
            Bakery.UpdateSubdeposits()
        Sleep(0.04)
        set TotalGameTime = TotalGameTime + 0.08
        #Try Produce Products
        for(MaybeBakery : AllBakers, Bakery := MaybeBakery?, Bakery.IsBakerReady? ,AgentOwner := Bakery.MaybeAgentOwner?, PlayerOwner:=player[AgentOwner], PlayerOwner.IsActive[]):
            Bakery.TryProduce_AllProducts()
        Sleep(0.04)
        set TotalGameTime = TotalGameTime + 0.08
        #Tick Time
        for(MaybeBakery : AllBakers, Bakery := MaybeBakery?, Bakery.IsBakerReady? ,AgentOwner := Bakery.MaybeAgentOwner?, PlayerOwner:=player[AgentOwner], PlayerOwner.IsActive[]):
            Bakery.ExecuteTickTime()

        <#
        for(Index:=0..GetPlayspace().GetPlayers().Length-1, MaybeBakery := AllBakers[Index], Bakery := MaybeBakery?):
            Bakery.FillMainDeposits()
            Bakery.TakeWater()
            Bakery.TakeFlour()
            Bakery.TakeCheese()
            Bakery.TakePimentos()
            Bakery.UpgradeValveButtonsDescriptions()
        Sleep(0.2)
        for(Index:=0..GetPlayspace().GetPlayers().Length-1, MaybeBakery := AllBakers[Index], Bakery := MaybeBakery?):
            Bakery.UpdateAllMainDeposits()
        Sleep(0.2)
        for(Index:=0..GetPlayspace().GetPlayers().Length-1, MaybeBakery := AllBakers[Index], Bakery := MaybeBakery?):
            Bakery.TryProduce_AllProducts()
        Sleep(0.2)
        for(Index:=0..GetPlayspace().GetPlayers().Length-1, MaybeBakery := AllBakers[Index], Bakery := MaybeBakery?):
            Bakery.ExecuteTickTime()#>
    

    #AGENT REMOVED AND ADDED
    #---------------------
    OnPlayerAdded(Player : player) : void =
        #Init gather UI:
        var PlayerNewData : player_data_table = player_data_table{}
        PlayersDataManager.InitializePlayer(Player)
        if(Agent := agent[Player]):
            if(PlayerData:=PlayersDataManager.LoadPlayerCache[Player]):
                set PlayerNewData = PlayerData
            else:
                PrintWarning("Failed on OnPlayerAdded, New player detected")
            #Creating inventory dat
            #Assign to Baker
            CalculeSuperEnergy()
            CheckAreasToUnlock(PlayerNewData,Agent)
            AssignToBaker(Agent)
            CollectiblesManager.InitPlayerCollectibles(Player)
            if(set AllAgentsJoinTime[Agent]=TotalGameTime):
            #Grant Arrows
            #GeneralMobsManager.ArrowsRemover.Remove(Agent)
            if(PlayerNewData.Ammo_Arrows>0):
                Log("Player Had arrows before, granting....", LOG_LEVEL_DEBUG7)
                for(Index:=0..PlayerNewData.Ammo_Arrows-1):
                    GeneralMobsManager.WeaponsGranter.SetNextItem(4)
                    GeneralMobsManager.WeaponsGranter.GrantItem(Agent)
            spawn: 
                DelayInventoryRefresh(Agent)
        else:
            PrintFatalError("OnPlayerAdded: Agent INVALID!")
        
        GeneralMobsManager.OnPlayerAdded(Player)
        

    DelayInventoryRefresh(Agent : agent)<suspends> : void =
        Sleep(10.0)
        Log("RefreshDelayExecuted!",LOG_LEVEL_DEBUG10)
        if(Player := player[Agent], Player.IsActive[], AgentBakery := GetAgentBakery_Transacted(Agent)?, MainInventoryUI:=AgentBakery.MaybeMainInventoryUI?):
            MainInventoryUI.Hide_All()
            Sleep(0.1)
            MainInventoryUI.SimplifiedRefreshAll()

    AssignToBaker(Agent : agent) : void =
        #Assign to empty slot
        var Index : int = 0
        ##Print("WARNING! Forzing the assignation of bakery 2")
        #set Index = 2
        loop:
            if:
                AllAgents[Index] = false #If it is empty this Index
                set AllAgents[Index] = option{Agent}
                Player := player[Agent]
                AgentData := PlayersDataManager.LoadPlayerCache[Player]
            then: 
                if:
                    MaybeBakery := AllBakers[Index]
                    ValidBakery := MaybeBakery?
                    set AllAgentsBakery[Agent] = ValidBakery #If the bakery exists, we reset assign it to the new player
                then:
                    Log("Using an existing bakery on [{Index}] to the new player", LOG_LEVEL_BAKERS)
                    ValidBakery.SetNewOwner(Agent,AgentData)
                    ValidBakery.DisplayBasicUIs()
                    spawn:
                        LateUIRefresh(ValidBakery)
                    break
                else: #If it is empty we create a new Bakery, if not, we use a existing one
                    Log("Assigning a new bakery [{Index}] to the new player", LOG_LEVEL_BAKERS)
                    if(HexaManager := AllHexaManager[Index]):
                        Bakery := MakeBakeryMachinesManager(Index,Agent,HexaManager,Self)
                        if(set AllBakers[Index] = option{Bakery}, set AllAgentsBakery[Agent] = Bakery):
                        #Print("Player and Bakery succesfully saved on {Index}")
                        spawn:
                            InitNewBakery(Bakery,Index) #Nope! Try to create this BEFORE assigning to a player
                        spawn:
                            LateUIRefresh(Bakery)
                        break
                    else:
                        PrintFatalError("AssignToBaker: AllHexaManager not found")
                        break
            else:
                set Index = Index + 1 
            if(Index > AllAgents.Length):
                #Print("FATAL ERROR: AllAgents Stackoverflow on AssignToBaker()")
                set Index = -1
                break

       
    InitNewBakery(BakeryMachinesManager : bakery_machines_manager, Index : int)<suspends> : void =
        #Set setting
        if(UKTutorial:=AllUKTutorials[Index]):
            UKTutorial.SetBakery(BakeryMachinesManager)
            BakeryMachinesManager.SetUKTutorial(UKTutorial)
        else:
            PrintFatalError("UKTutorial: {Index} not found!")
        BakeryMachinesManager.SetMainDepositsProps(PropsManager.GetMainDepositsProps(Index))
        BakeryMachinesManager.SetValves(PropsManager.GetValves(Index))
        BakeryMachinesManager.SetTeleportationDevices(PropsManager.GetTeleportDevices(Index))
        BakeryMachinesManager.SetUpgradeButtons(PropsManager.GetUpgradesButtons(Index))
        BakeryMachinesManager.SetBillBoards(PropsManager.GetBillboards(Index))
        BakeryMachinesManager.SetSuperEnergyDevices(PropsManager.GetSuperEnergyDevices(Index))
        BakeryMachinesManager.SetBakersProps(PropsManager.GetBakersProps(Index))
        BakeryMachinesManager.SetFunnelButton(PropsManager.GetFunnelButton(Index))
        BakeryMachinesManager.SetPlayerReference(PropsManager.GetPlayerReferenceDevice(Index))

        Sleep(2.0)
        BakeryMachinesManager.SetPropTransforms()
        BakeryMachinesManager.InitBakery()
        #LoadAgentBakerProp()
        BakeryMachinesManager.CheckMusicPlay(BAKERSMUSIC_SOUND)
        BakeryMachinesManager.RefreshAllUI()

    #PetManager
    Display_PetManager(Agent : agent) : void =
        MaybeBakery:=GetAgentBakery(Agent)
        if(ValidBakery:=MaybeBakery?, ValidBakery.PlayerDataCache.NumQuestsCompleted>=27):
            ValidBakery.PetsManager.MainMenu_Display_Canvas()
        else:
            DisplayCharacterMessage(Agent, I_CHARACTER, "Complete the tutorial to unlock the Sourdoughs!", ?Sound:=ERROR_SOUND)
    #Backpack Upgrades
    Display_BackpackUpgrades(Agent : agent) : void =
        MaybeBakery := GetAgentBakery(Agent)
        if:
            Bakery := MaybeBakery?
            ValidOwner := Bakery.MaybeAgentOwner?
            ValidOwner = Agent
            BackpackUpgradesUI := Bakery.MaybeBackpackUpgradesUI?
            PlayerOwner:=player[ValidOwner]
        then:
            BackpackUpgradesUI.Display_Canvas()
            BackpackUpgradesUI.UpdateUI(Bakery.PlayerDataCache)
        else:
            Display_NotOwner(Agent)

    #Pickaxe upgrades Upgrades
    Display_PickaxeUpgrades(Agent : agent) : void =
        MaybeBakery := GetAgentBakery(Agent)
        if:
            Bakery := MaybeBakery?
            ValidOwner := Bakery.MaybeAgentOwner?
            ValidOwner = Agent
            PickaxeUpgradesUI := Bakery.MaybePickaxeUpgradesUI?
            PlayerOwner:=player[ValidOwner]
        then:
            PickaxeUpgradesUI.Display_Canvas()
            PickaxeUpgradesUI.UpdateUI(Bakery.PlayerDataCache,Bakery.PickaxeTotalStrength)
        else:
            Display_NotOwner(Agent)

    #Pickaxe upgrades Upgrades perks
    Display_PickaxeUpgradesPerks(Agent : agent) : void =
        MaybeBakery := GetAgentBakery(Agent)
        if:
            Bakery := MaybeBakery?
            ValidOwner := Bakery.MaybeAgentOwner?
            ValidOwner = Agent
            PickaxePerksUpgradesUI := Bakery.MaybePickaxePerksUI?
            PlayerOwner:=player[ValidOwner]
        then:
            PickaxePerksUpgradesUI.Display_Canvas()
            PickaxePerksUpgradesUI.UpdateUI(Bakery.PlayerDataCache)
        else:
            Display_NotOwner(Agent)
        
    OnPlayerRemoved(Player : player) : void =
        if(Agent := agent[Player]):
            AuxMap := RemoveAgentFromFloatMap(AllAgentsJoinTime,Agent)
            set AllAgentsJoinTime = AuxMap
            var Index : int = 0
            loop:
                if:
                    AllAgents[Index] = option{Agent}, #Agent found
                    set AllAgents[Index] = false,   #Agent deleted
                    MaybeBaker := AllBakers[Index] #Baker found
                then:
                    MiniGamesManager.OnPlayerRemoved(Index)
                    NewMapAllAgentsBakery:=RemoveAgentAssignedToBakery(AllAgentsBakery,Agent)
                    set AllPlayersAFKRegister = RemoveAgentFromFloatMap(AllPlayersAFKRegister,Agent)
                    set AllAgentsBakery = NewMapAllAgentsBakery
                    #Print("Player Index {Index}, removed from Baker {Index}")
                    if(Baker := MaybeBaker?):
                        if(Baker.SuperEnergyRemaining_Timer.GetMaxDuration()>1.0):
                            CalculeSuperEnergy()
                            #set SuperEnergy_Shared = SuperEnergy_Shared - Baker.SuperEnergyRemaining_Timer.GetMaxDuration()
                        Baker.OnPlayerRemoved() #Reset of the bakery
                    MiniGamesManager.RemoveScroll(Index)
                    break
                else:
                    set Index = Index + 1
                if(Index > AllAgents.Length):
                    #Print("FATAL ERROR: AllAgents Stackoverflow on OnPlayerRemoved(). Player removed NOT found")
                    set Index = -1
                    break
        GeneralMobsManager.OnPlayerRemoved(Player)

            
    #LoadAgentBakerProp() : void =
    #    Print("WARNING TO DO: LoadAgentBaker() to load the props of the player Baker")

    #RemoveAgentBakerProp(Index : int) : void =
    #    Print("WARNING TO DO: RemoveAgentBaker() to reset all the props of the Baker {Index} that has just been abandoned")
    #---------------------
    DisplayCharacterMessage(Agent : agent, Character : int, StringMessage : string, ?Sound:int=-1, ?DisplayTime : float = 5.0, ?Priority:int=MIN_PRIORITY) : void =
        MaybeBakery := GetAgentBakery(Agent)
        if(Player := player[Agent], Player.IsActive[], ValidBakery:=MaybeBakery?, MainInventoryUI := ValidBakery.MaybeMainInventoryUI?):
            MainInventoryUI.DisplayCharacterMessage(Character,option{StringMessage},false,Sound,?DisplayTime:=DisplayTime,?Priority:=Priority) 

    DisplayCharacterMessageLoc(Agent : agent, Character : int, InMessage : message, ?Sound:int=-1, ?DisplayTime : float = 5.0, ?Priority:int=MIN_PRIORITY) : void =
        MaybeBakery := GetAgentBakery(Agent)
        if(Player := player[Agent], Player.IsActive[], ValidBakery:=MaybeBakery?, MainInventoryUI := ValidBakery.MaybeMainInventoryUI?):
            MainInventoryUI.DisplayCharacterMessage(Character,false,option{InMessage},Sound,?DisplayTime:=DisplayTime,?Priority:=Priority) 

    CleanCharacterMessages(Agent : agent) : void =
        MaybeBakery := GetAgentBakery(Agent)
        if(ValidBakery:=MaybeBakery?, MainInventoryUI:=ValidBakery.MaybeMainInventoryUI?):
            MainInventoryUI.CleanCharacterMessages()

#Player stats
    #GATHER BASE (Only for auto pipes)
    AddPlayerTo_RecentsAgentsJoined(Agent : agent) : void =
        set RecentAgentsJoined = RecentAgentsJoined + array{Agent}

    Display_AlreadyUnlocked<public>(Agent : agent) : void = 
        DisplayCharacterMessage(Agent,H_CHARACTER,"Hey buddy, you already unlocked this area!")

    Display_UnlockPrevious<public>(Agent : agent) : void = 
        DisplayCharacterMessage(Agent,H_CHARACTER,"Hey buddy, unlock the previous area first!")
        
    Display_NotEnough<public>(Agent : agent, AmountNeeded : float) : void = 
        DisplayCharacterMessage(Agent,H_CHARACTER,"Hey buddy! You need {GetOptimizedMagnitudeLetter(AmountNeeded)} more Sapphires!")

    Display_NotOwner<public>(Agent : agent) : void = 
        DisplayCharacterMessage(Agent,I_CHARACTER,"You are not the owner of this Bakery!")

    Display_DroneNotReady(Agent : agent, Time : int) : void = 
        DisplayCharacterMessage(Agent,I_CHARACTER, "Drone not ready, wait {Time} more seconds!")

    CheckZoneMusic(Agent : agent, Music : int) : void =
        if(ValidBakery := GetAgentBakery_Transacted(Agent)?):
            ValidBakery.CheckMusicPlay(Music)

    ChangeEnglishLanguage(Agent : agent) : void =
        if(ValidBakery:=GetAgentBakery_Transacted(Agent)?, LanguageUI := ValidBakery.MaybeLanguageUI?):
            LanguageUI.SwitchToEnglish_N()

    ChangeSpanishLanguage(Agent : agent) : void =
        if(ValidBakery := GetAgentBakery_Transacted(Agent)?, LanguageUI := ValidBakery.MaybeLanguageUI?):
            LanguageUI.SwitchToSpanish_N()
    
    ChangeFrenchLanguage(Agent : agent) : void =
        if(ValidBakery := GetAgentBakery_Transacted(Agent)?, LanguageUI := ValidBakery.MaybeLanguageUI?):
            LanguageUI.SwitchToFrench_N()
    
    ChangePortugueseLanguage(Agent : agent) : void =
        if(ValidBakery := GetAgentBakery_Transacted(Agent)?, LanguageUI := ValidBakery.MaybeLanguageUI?):
            LanguageUI.SwitchToPortuguese_N()
            
    InitSubscribers() : void =

        EnterUKTutorialButton.InteractedWithEvent.Subscribe(EnterOnUKTutorial)

        English_MutatorsZone.AgentEntersEvent.Subscribe(ChangeEnglishLanguage)
        Spanish_MutatorsZone.AgentEntersEvent.Subscribe(ChangeSpanishLanguage)
        French_MutatorsZone.AgentEntersEvent.Subscribe(ChangeFrenchLanguage)
        Portuguese_MutatorsZone.AgentEntersEvent.Subscribe(ChangePortugueseLanguage)

        #MusicZonesChecks
        for(MutatorZone : AllBakerMusic_MutatorsZones):
            MutatorZone.AgentEntersEvent.Subscribe(handler_random{MaybeGameManager:=option{Self},Index:=BAKERSMUSIC_SOUND}.HandlerCheckMusic)

        for(MutatorZone : AllDesertMusic_MutatorsZones):
            MutatorZone.AgentEntersEvent.Subscribe(handler_random{MaybeGameManager:=option{Self},Index:=DESERTMUSIC_SOUND}.HandlerCheckMusic)

        for(MutatorZone : AllSecretMusic_MutatorsZones):
            MutatorZone.AgentEntersEvent.Subscribe(handler_random{MaybeGameManager:=option{Self},Index:=SECRETMUSIC_SOUND}.HandlerCheckMusic)

        Button_DebugUI.InteractedWithEvent.Subscribe(HandlerDebugAllUIs)
        
        MutZone_AfkZone.AgentEntersEvent.Subscribe(AgentEntersAFKZone)
        MutZone_AfkZone.AgentExitsEvent.Subscribe(AgentExitsAFKZone)
        #Signal remote manager
        InputTriggerOptions.PressedEvent.Subscribe(Display_OptionsMenu)
        #SignalRemoteManager.PrimarySignalEvent.Subscribe(Display_OptionsMenu)
        #SignalRemoteManager.SecondarySignalEvent.Subscribe(Display_OptionsMenu)

        if(DebugOn?):
            #Debug
            Test_AddPhase_Button.InteractedWithEvent.Subscribe(Test_AddTutorialPhase)
            Test_UpdateBackpackButton.InteractedWithEvent.Subscribe(Test_UpgradeBackpack)
            Test_LoadData_Button.InteractedWithEvent.Subscribe(Test_LoadDataTable)
            Test_GrantResourcers_Button.InteractedWithEvent.Subscribe(Test_GrantResourcers)
            Test_GrantSapphires_Button.InteractedWithEvent.Subscribe(Test_GrantSapphires)
            Test_GrantCoal_Button.InteractedWithEvent.Subscribe(Test_GrantCoal)
            Test_GrantAgedSapphires_Button.InteractedWithEvent.Subscribe(Test_GrantAgedSapphires)
            Test_GrantMutantMaterial_Button.InteractedWithEvent.Subscribe(Test_GrantMutantMaterial)

            Test_AddPermision_Button.InteractedWithEvent.Subscribe(Test_AddUnlockHexaChallengeLV)
            Test_GrantOrderProgress_Button.InteractedWithEvent.Subscribe(Test_AddProgressToOrder)
            Test_CompleteTutorialButton.InteractedWithEvent.Subscribe(Test_CompleteQuest)
            Test_CompleteAllTutorialButton.InteractedWithEvent.Subscribe(Test_CompleteAllTutorials)

        Button_ImproveDungeonTokens.InteractedWithEvent.Subscribe(DisplayImproveDungeonTokens)
        Button_ImproveHexaGameTokens.InteractedWithEvent.Subscribe(DisplayImproveHexaGameTokens)
        Button_ImproveHammer.InteractedWithEvent.Subscribe(DisplayImproveHammer)
        Button_ImproveBow.InteractedWithEvent.Subscribe(DisplayImproveBow)

        Button_BuyArrows.InteractedWithEvent.Subscribe(TryBuyArrows)
        
        Button_HexaGame.InteractedWithEvent.Subscribe(PlayHexaGame)
        for(Button : AllButtons_InserCoal):
            Button.InteractedWithEvent.Subscribe(DisplayCoalUI)

        Button_PetManager.InteractedWithEvent.Subscribe(Display_PetManager)

        Button_SkipQuest.InteractedWithEvent.Subscribe(TryCompleteQuest_Paid)

        Button_Mastery.InteractedWithEvent.Subscribe(TryGetMastery)
        if(DebugOn?):
            Button_CompleteReset.InteractedWithEvent.Subscribe(ResetPlayer)

        GetPlayspace().PlayerAddedEvent().Subscribe(AddPlayerTo_RecentsAgentsJoined)
        GetPlayspace().PlayerRemovedEvent().Subscribe(OnPlayerRemoved)

        Button_UpgradeBackpack.InteractedWithEvent.Subscribe(Display_BackpackUpgrades)
        Button_UpgradePickaxe.InteractedWithEvent.Subscribe(Display_PickaxeUpgrades)
        Button_UpgradePickaxePerks.InteractedWithEvent.Subscribe(Display_PickaxeUpgradesPerks)

        for(Index:=0..AllButtonsArea.Length-1, Button := AllButtonsArea[Index]):
            Button.InteractedWithEvent.Subscribe(handler_upgrades{MaybeGameManager:=option{Self},Index:=Index}.HandlerTryUnlockArea)

        for(Index:=0..AllWeakPoints_Manipulators.Length-1, Manipulator := AllWeakPoints_Manipulators[Index]):
            Manipulator.HarvestingEvent.Subscribe(handler_random{MaybeGameManager:=option{Self},Index:=Index}.HandlerWeakPointEvent)
            #Manipulator.DamagedEvent.Subscribe(handler_random{MaybeGameManager:=option{Self},Index:=Index}.HandlerWeakPointEvent)

    InitSpawnerSubscribers() : void =
        for(Spawner : AllPlayerSpawners):
            Spawner.SpawnedEvent.Subscribe(OnPlayerSpawn)