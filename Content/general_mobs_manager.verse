using { /Fortnite.com/Devices }
using { /Verse.org/Simulation }
using { /UnrealEngine.com/Temporary/Diagnostics }
using { /Fortnite.com/Characters }
using { /UnrealEngine.com/Temporary/SpatialMath }
using { /Verse.org/Random }
using { /Fortnite.com/Game }

# See https://dev.epicgames.com/documentation/en-us/uefn/create-your-own-device-in-verse for how to create a verse device.

# A Verse-authored creative device that can be placed in a level
general_mobs_manager := class(creative_device):

    var TotalGameTime : float = 0.0

    var MaybeGameManager : ?game_manager = false
    DungeonDataBase : dungeon_database = dungeon_database{}
    @editable
    MobBehaviourHelper : mob_behavior_helper = mob_behavior_helper{}

    @editable
    WeaponsGranter : item_granter_device = item_granter_device{}
    @editable
    WeaponsRemover : item_remover_device = item_remover_device{}
    @editable
    ArrowsRemover : item_remover_device = item_remover_device{}

    @editable
    DirtSpawnEffect : creative_prop_asset = DefaultCreativePropAsset

    @editable
    HealthBarIsOrganic : logic = true

    #@editable
    #WeaponsManager : weapons_manager = weapons_manager{}

    var AllMobsManager : []mob = array{}
    var AllRandomEncounterMobInfoToLoad : []tuple(transform,logic,int,int,int,logic,int,?player) = array{}


    #RandomSpawner
    @editable
    AllNPC_PlantMobSpawner_RandomEncounter : []npc_spawner_device = array{}

    #Ground spawners
    var AllNPC_GroundMobSpawner_IsBeingUsed : []int = array{}
    @editable
    AllNPC_GroundMobSpawner : []npc_spawner_device = array{}
    #Druid Spawner
    var AllNPC_DruidMobSpawner_IsBeingUsed : []int = array{}
    @editable
    AllNPC_DruidMobSpawner : []npc_spawner_device = array{}
    #Plant Spawner
    var AllNPC_PlantMobSpawner_IsBeingUsed : []int = array{}
    @editable
    AllNPC_PlantMobSpawner : []npc_spawner_device = array{}
    #MiniBossSpawner
    var AllNPC_MiniBossMobSpawner_IsBeingUsed : []int = array{}
    @editable
    AllNPC_MiniBossMobSpawner : []npc_spawner_device = array{}
    
    #AllDungeons
    var AllDungeons_ExpectedHammerDamage : []float = array{}
    var AllDungeons_MobBaseDamage : []float = array{}

    var AllDungeons_FinalMusicIsPlaying : []logic = array{}
    var Dungeon_TresureRegistration : [agent][]int = map{}
    var AllDungeons_IsPublic : []logic = array{}
    var AllDungeons_MobsSpawned : [][]int = array{}
    var AllDungeons_MobsDistributionToBeSpawn : [][]int = array{}
    var AllDungeons_NPCSpawnersInUse : [][]int = array{}
    var AllDungeons_ChestBeaconsInUse : [][]int = array{}

    var AllDungeons_MinibossIsDefeated : []logic = array{}
    var AllDungeons_MaybeStartingAgent : []?agent = array{}
    var AllDungeons_HasStarted : []logic = array{} 
    var AllDungeons_IsReadyToPlay : []logic = array{}
    var AllDungeons_CountToStartTime : []int = array{}
    var AllDungeons_TotalTime : []int = array{}
    var AllDungeons_Buttons : [][]button_device = array{}
    var AllDungeons_Positions : [][]creative_prop = array{} 
    var AllDungeons_AgentsWaitingToJoin : [][]agent = array{}
    var AllDungeons_AgentsPlaying : [][]agent = array{}
    var AllDungeons_AllMobInfoToLoad : [][]tuple(transform,logic,int,int,int,logic,int) = array{}
    var AllDungeons_AllMobBossesInfoToLoad : [][]tuple(transform,logic,int,int,int,logic,int) = array{}
    var AllDungeons_HBAreaLevel : []int = array{}

    #Chests
    @editable
    AllDungeonsChestBeacons : []beacon_device = array{}
    var AllDungeonsChestBeacons_IsUsed : []logic = array{}

    #Positions! 
    #0 Position: Entry transform
    #1 Position: Exit transform
    #2 Position: Boss spawner transform
    #3..X: Posible Mob spawner transform

    #Buttons!
    #0 Position: Entry button
    #1 Position: Exit button
    #2 Position: Main treasure button
    #3..X: Position: Other treasures button

    #Dungeon 0
    var Dungeon0_TotalTime : int = 0
    @editable
    AllDungeon0_Positions : []creative_prop = array{}
    @editable
    AllDungeon0_Buttons : []button_device = array{}
    var AllDungeon0_ChestBeacon : []int = array{}
    var AllDungeon0_NPCSpawnerUsed : []int = array{-1,-1,-1,-1}

    var IsDungeon0BeingPlayed : logic = true
    var Dungeon0_MobsDistribution : []int = array{}
    var Dungeon0_MobsSpawned : []int = array{}
    var Dungeon0_AgentsPlaying : []agent = array{}
    var Dungeon0_AgentsWaiting : []agent = array{}

    var Dungeon0_AllMobInfoToLoad : []tuple(transform,logic,int,int,int,logic,int) = array{}
    var Dungeon0_AllMobBossesInfoToLoad : []tuple(transform,logic,int,int,int,logic,int) = array{}

    #Dungeon 1
    var Dungeon1_TotalTime : int = 0
    @editable
    AllDungeon1_Positions : []creative_prop = array{}
    @editable
    AllDungeon1_Buttons : []button_device = array{}
    var AllDungeon1_ChestBeacon : []int = array{}
    var AllDungeon1_NPCSpawnerUsed : []int = array{-1,-1,-1,-1}

    var IsDungeon1BeingPlayed : logic = true
    var Dungeon1_MobsDistribution : []int = array{}
    var Dungeon1_MobsSpawned : []int = array{}
    var Dungeon1_AgentsPlaying : []agent = array{}
    var Dungeon1_AgentsWaiting : []agent = array{}

    var Dungeon1_AllMobInfoToLoad : []tuple(transform,logic,int,int,int,logic,int) = array{}
    var Dungeon1_AllMobBossesInfoToLoad : []tuple(transform,logic,int,int,int,logic,int) = array{}

    #Dungeon 2
    var Dungeon2_TotalTime : int = 0
    @editable
    AllDungeon2_Positions : []creative_prop = array{}
    @editable
    AllDungeon2_Buttons : []button_device = array{}
    var AllDungeon2_ChestBeacon : []int = array{}
    var AllDungeon2_NPCSpawnerUsed : []int = array{-1,-1,-1,-1}

    var IsDungeon2Available : logic = true
    var Dungeon2_MobsDistribution : []int = array{}
    var Dungeon2_MobsSpawned : []int = array{}
    var Dungeon2_AgentsPlaying : []agent = array{}
    var Dungeon2_AgentsWaiting : []agent = array{}

    var Dungeon2_AllMobInfoToLoad : []tuple(transform,logic,int,int,int,logic,int) = array{}
    var Dungeon2_AllMobBossesInfoToLoad : []tuple(transform,logic,int,int,int,logic,int) = array{}

    #Dungeon 3
    var Dungeon3_TotalTime : int = 0
    @editable
    AllDungeon3_Positions : []creative_prop = array{}
    @editable
    AllDungeon3_Buttons : []button_device = array{}
    var AllDungeon3_ChestBeacon : []int = array{}
    var AllDungeon3_NPCSpawnerUsed : []int = array{-1,-1,-1,-1}

    var IsDungeon3BeingPlayed : logic = true
    var Dungeon3_MobsDistribution : []int = array{}
    var Dungeon3_MobsSpawned : []int = array{}
    var Dungeon3_AgentsPlaying : []agent = array{}
    var Dungeon3_AgentsWaiting : []agent = array{}

    var Dungeon3_AllMobInfoToLoad : []tuple(transform,logic,int,int,int,logic,int) = array{}
    var Dungeon3_AllMobBossesInfoToLoad : []tuple(transform,logic,int,int,int,logic,int) = array{}

    #Dungeon 4
    var Dungeon4_TotalTime : int = 0
    @editable
    AllDungeon4_Positions : []creative_prop = array{}
    @editable
    AllDungeon4_Buttons : []button_device = array{}
    var AllDungeon4_ChestBeacon : []int = array{}
    var AllDungeon4_NPCSpawnerUsed : []int = array{-1,-1,-1,-1}

    var IsDungeon4BeingPlayed : logic = true
    var Dungeon4_MobsDistribution : []int = array{}
    var Dungeon4_MobsSpawned : []int = array{}
    var Dungeon4_AgentsPlaying : []agent = array{}
    var Dungeon4_AgentsWaiting : []agent = array{}

    var Dungeon4_AllMobInfoToLoad : []tuple(transform,logic,int,int,int,logic,int) = array{}
    var Dungeon4_AllMobBossesInfoToLoad : []tuple(transform,logic,int,int,int,logic,int) = array{}

    #Dungeon 5
    var Dungeon5_TotalTime : int = 0
    @editable
    AllDungeon5_Positions : []creative_prop = array{}
    @editable
    AllDungeon5_Buttons : []button_device = array{}
    var AllDungeon5_ChestBeacon : []int = array{}
    var AllDungeon5_NPCSpawnerUsed : []int = array{-1,-1,-1,-1}

    var IsDungeon5BeingPlayed : logic = true
    var Dungeon5_MobsDistribution : []int = array{}
    var Dungeon5_MobsSpawned : []int = array{}
    var Dungeon5_AgentsPlaying : []agent = array{}
    var Dungeon5_AgentsWaiting : []agent = array{}

    var Dungeon5_AllMobInfoToLoad : []tuple(transform,logic,int,int,int,logic,int) = array{}
    var Dungeon5_AllMobBossesInfoToLoad : []tuple(transform,logic,int,int,int,logic,int) = array{}

    #Dungeon 6
    var Dungeon6_TotalTime : int = 0
    @editable
    AllDungeon6_Positions : []creative_prop = array{}
    @editable
    AllDungeon6_Buttons : []button_device = array{}
    var AllDungeon6_ChestBeacon : []int = array{}
    var AllDungeon6_NPCSpawnerUsed : []int = array{-1,-1,-1,-1}

    var IsDungeon6BeingPlayed : logic = true
    var Dungeon6_MobsDistribution : []int = array{}
    var Dungeon6_MobsSpawned : []int = array{}
    var Dungeon6_AgentsPlaying : []agent = array{}
    var Dungeon6_AgentsWaiting : []agent = array{}

    var Dungeon6_AllMobInfoToLoad : []tuple(transform,logic,int,int,int,logic,int) = array{}
    var Dungeon6_AllMobBossesInfoToLoad : []tuple(transform,logic,int,int,int,logic,int) = array{}


    #Dungeon 7
    var Dungeon7_TotalTime : int = 0
    @editable
    AllDungeon7_Positions : []creative_prop = array{}
    @editable
    AllDungeon7_Buttons : []button_device = array{}
    var AllDungeon7_ChestBeacon : []int = array{}
    var AllDungeon7_NPCSpawnerUsed : []int = array{-1,-1,-1,-1}

    var IsDungeon7BeingPlayed : logic = true
    var Dungeon7_MobsDistribution : []int = array{}
    var Dungeon7_MobsSpawned : []int = array{}
    var Dungeon7_AgentsPlaying : []agent = array{}
    var Dungeon7_AgentsWaiting : []agent = array{}

    var Dungeon7_AllMobInfoToLoad : []tuple(transform,logic,int,int,int,logic,int) = array{}
    var Dungeon7_AllMobBossesInfoToLoad : []tuple(transform,logic,int,int,int,logic,int) = array{}

    #WEAPONS

    PlayersDataManager : player_data_manager = player_data_manager{}

    #@editable
    #Sentry : sentry_device = sentry_device{}

    #Bow checker
    @editable
    InputTriggerFire : input_trigger_device = input_trigger_device{}

    @editable
    SwordEquiped_ConditionalButton : conditional_button_device = conditional_button_device{}
    @editable
    BowEquiped_ConditionalButton : conditional_button_device = conditional_button_device{}
    @editable
    HowManyArros_ConditionalButton : conditional_button_device = conditional_button_device{}


    @editable
    SuperStrikeEffect : visual_effect_powerup_device = visual_effect_powerup_device{}
    @editable
    RemoveSuperStrikeEffect : visual_effect_powerup_device = visual_effect_powerup_device{}

    @editable
    AllKnifeEffects : []visual_effect_powerup_device = array{}
    @editable
    RemoveKnifeEffect : visual_effect_powerup_device = visual_effect_powerup_device{}
    var AllPlayersKnifeEffects : []int = array{}

    #@editable
    #Damage_Zero : damage_amplifier_powerup_device = damage_amplifier_powerup_device{}

    SetGameManager(GameManager : game_manager) : void =
        set MaybeGameManager = option{GameManager}

    GrantMobDestroyedReward(MaybeEliminatorPlayer : ?player, AmountDamaged : float, MaxHealth : float, MobType : int, AreaLV : int,  NumDungeon : int, HBAreaLevel : int ) : void =
        if(GameManager:=MaybeGameManager?):
            var ResourceAmount : float = CheckFinalMutantMaterialAward(HBAreaLevel,15.0,1.0,?MobType:=MobType,?NumDungeon:=NumDungeon)
            ResourceType := MUTANTMATERIAL
            if(NumDungeon=-1, EliminatorPlayer:=MaybeEliminatorPlayer?, Agent := agent[EliminatorPlayer], EliminatorPlayer.IsActive[]):
                PercentageDamaged:=AmountDamaged/MaxHealth
                if(PercentageDamaged>1.0):
                    set ResourceAmount = ResourceAmount *1.0
                else:
                    set ResourceAmount = ResourceAmount * PercentageDamaged
                Log("PercentageDamaged:  {GetOptimizedMagnitudeLetter(PercentageDamaged)} {GetOptimizedMagnitudeLetter(AmountDamaged)} of a total of {MaxHealth}", LOG_LEVEL_DEBUG4)
                MaybeBakery:=GameManager.GetAgentBakery(Agent)
                if(ValidBakery:=MaybeBakery?):
                    AllLimitsReached :=ValidBakery.SaveResourcers(Agent,array{ResourceType},array{ResourceAmount})
                    ValidBakery.RefreshInventoryUI(ValidBakery.PlayerDataCache)
                    ValidBakery.DisplayDamageUI(Agent, array{ResourceAmount},array{ResourceType}, AllLimitsReached,false)
            else:
                #If it is a dungeon all player receive the same reward
                if(AllAgentsPlaying:=AllDungeons_AgentsPlaying[NumDungeon]):
                    for(PlayingAgent : AllAgentsPlaying, PlayingPlayer := player[PlayingAgent], PlayingPlayer.IsActive[]):
                        MaybeBakery:=GameManager.GetAgentBakery(PlayingPlayer)
                        if(ValidBakery:=MaybeBakery?):
                            AllLimitsReached :=ValidBakery.SaveResourcers(PlayingAgent,array{ResourceType},array{ResourceAmount})
                            ValidBakery.RefreshInventoryUI(ValidBakery.PlayerDataCache)
                            ValidBakery.DisplayDamageUI(PlayingAgent, array{ResourceAmount},array{ResourceType}, AllLimitsReached,false)
                else:
                    PrintFatalError("AllAgentsPlaying Dungeon: {NumDungeon} not found")
            Log("AreaLV: {AreaLV}, NumDungeon: {NumDungeon} ResourceAmount: {GetOptimizedMagnitudeLetter(ResourceAmount)} of Mutant Material",LOG_LEVEL_DEBUG4)
        else:
            PrintFatalError("Invalid player eliminator")
                
    # Runs when the device is started in a running game
    OnBegin<override>()<suspends>:void=
        #for(Player : GetPlayspace().GetPlayers(), Agent := agent[Player], Character := Agent.GetFortCharacter[]):
            #InitPlayer(Agent)
            #AddAgentToChestRegistration(Agent)
            #Character.EliminatedEvent().Subscribe(OnPlayerEliminated)

        InitOrganizationArrays()
        #GetPlayspace().PlayerRemovedEvent().Subscribe(OnPlayerRemoved)
        #PrintWarning("TODO OnPlayerEliminated on general mobs manager for dungeons")
        
        CreateMobsManagers()
        InitNPCSpawners()
        InitSubscribers()
        #MoveVisualEmpty()
        RefreshSwordEffects()

        MobBehaviourHelper.SetGeneralMobManager(Self)

        #spawn:
            #LoopGameTime()

    RollRandomPlantEncounter(Agent : agent, Transform : transform, IsBoss : logic, MobLevel : int, NumHits : int, HBAreaLevel : int) : logic =
        var ThereIsARandomEncounter : logic = false
        if(GameManager:=MaybeGameManager?):
            MaybeBakery:=GameManager.GetAgentBakery(Agent)
            if:
                ValidBakery := MaybeBakery?
                PetManager := ValidBakery.PetsManager
            then:
                HitsMultiplierProb := (NumHits*1.0)/MinNum_AVGHit
                #PrintWarning("Todo calculate total posibility based on RandomPlantEncounter_PosibilityBase RandomPlantEncounter_Posibility_PerHitAdded RandomPlantEncounter_MaxBasePosibility")
                var TotalProbability : float = PetManager.PlayerEnhacement_MobHunter*RandomPlantEncounter_PosibilityBase*HitsMultiplierProb
                Log("HitsMultiplierProb: {HitsMultiplierProb}", LOG_LEVEL_DEBUG7)
                Log("RandomPlantEncounter_PosibilityBase: {RandomPlantEncounter_PosibilityBase}", LOG_LEVEL_DEBUG7)
                Log("MobHunter Pet Multiplier: {PetManager.PlayerEnhacement_MobHunter}", LOG_LEVEL_DEBUG7)
                Log("TotalProbability: {TotalProbability}", LOG_LEVEL_DEBUG7)
                RandomFloat := GetRandomFloat(0.0,1.0)
                Log("RandomFloat: {RandomFloat}", LOG_LEVEL_DEBUG7)
                if(GameManager.DebugOn?, LOG_LEVEL = LOG_LEVEL_DEBUG9, PetManager.PlayerEnhacement_MobHunter>0.0):
                    #Print("Forced TotalProbability = 9999.9")
                    set TotalProbability = 9999.9
                if(RandomFloat<=TotalProbability):
                    Log("Success! RandomFloat: {RandomFloat}, spawning mob", LOG_LEVEL_DEBUG5)
                    spawn:
                        ExecuteEncounter_PlantMobSpawn(Agent,Transform,IsBoss,MobLevel,HBAreaLevel)
                    set ThereIsARandomEncounter = true
                else:
                    set ThereIsARandomEncounter = false
            else:
                PrintFatalError("On RollRandomPlantEncounter")
                set ThereIsARandomEncounter = false
        else:
            PrintFatalError("On RollRandomPlantEncounter")
            set ThereIsARandomEncounter = false
        return ThereIsARandomEncounter

    GranAgentWeapons(Agent : agent)<suspends> : void =
        if:
            Player := player[Agent]
            Player.IsActive[]
            Character := Agent.GetFortCharacter[]
            GameManager:=MaybeGameManager?
        then:
            MaybeBakery:=GameManager.GetAgentBakery(Agent)
            if(Bakery:=MaybeBakery?):
                #Character.SetMaxHealth(PlayerBaseHealth)
                WeaponsRemover.Remove(Agent)
                Sleep(0.1)
                #Grant sword
                WeaponsGranter.SetNextItem(0)
                WeaponsGranter.GrantItem(Agent)
                Sleep(0.1)
                #Bow
                WeaponsGranter.SetNextItem(Bakery.PlayerDataCache.Tier_Bow)
                WeaponsGranter.GrantItem(Agent)
                #Signal device
                #WeaponsGranter.SetNextItem(6)
                #WeaponsGranter.GrantItem(Agent)

    <#
    MoveVisualEmpty(): void =
        for(AllDungeonPositions : AllDungeons_Positions):
            for(Position : AllDungeonPositions):
                NewTranslationX:=Position.GetTransform().Translation.X
                NewTranslationY:=Position.GetTransform().Translation.Y
                NewTranslationZ:=Position.GetTransform().Translation.Z-LowVisualEmpty
                NewTranslation:=vector3{X:=NewTranslationX,Y:=NewTranslationY,Z:=NewTranslationZ}
                if(Position.TeleportTo[transform{Translation:=NewTranslation,Rotation:=Position.GetTransform().Rotation,Scale:=Position.GetTransform().Scale}]):#>

                <#
    LoopGameTime()<suspends> : void =
        #PrintWarning("TODO: Loop game time")
        loop:
            CheckNPCSpawners()
        #>

    CheckWeaponsEffects() : void = 
        AllPlayers := GetPlayspace().GetPlayers()
        for(Index:=0..AllPlayers.Length-1, KnifeIndex:=AllPlayersKnifeEffects[Index], Player:=AllPlayers[Index], Agent:=agent[Player], Effect:=AllKnifeEffects[KnifeIndex]):
            if(SwordEquiped_ConditionalButton.IsHoldingItem[Agent]):
                Effect.Pickup(Agent)
            else:
                RemoveKnifeEffect.Pickup(Agent)

    OnPlayerEliminated(Result : elimination_result) : void =
        Eliminated := Result.EliminatedCharacter
        if(EliminatedFortCharacter := Eliminated, EliminatedAgent := EliminatedFortCharacter.GetAgent[], GameManager := MaybeGameManager?):
            Log("Player eliminated detected",LOG_LEVEL_DEBUG1)
            NumDungeon:=CheckIfIsPlaying(EliminatedAgent)
            if(NumDungeon<>-1):
                spawn:
                    GameManager.VFXManager.PlaySound(LOSE_SOUND, 0.0, option{EliminatedAgent}, false)
                RandomInt:=GetRandomInt(0,1)
                if(1=GetRandomInt(0,1)):
                    GameManager.DisplayCharacterMessage(EliminatedAgent, G_CHARACTER,"Better luck next time!")
                else:
                    GameManager.DisplayCharacterMessage(EliminatedAgent, H_CHARACTER,"Better luck next time buddy!")
                QuickAgentFromDungeon(EliminatedAgent,NumDungeon)
                ApplyDeathPenalization(EliminatedAgent,true)
            else:
                ApplyDeathPenalization(EliminatedAgent,false)
            MaybeBakery:=GameManager.GetAgentBakery(EliminatedAgent)
            if(ValidBakery := MaybeBakery?, PetsManager:=ValidBakery.PetsManager):
                PetsManager.Generate_PlayerEnhacements()
            
    ApplyDeathPenalization<private>(Agent : agent, WasOnDungeon : logic) : void =
        AVGRunSpeed := 600.0 #Six hundred centimeters per second
        AVGHit := 1.5 # 1.5 hits per second
        if:
            GameManager:=MaybeGameManager?
            Teleporter0:=GameManager.TeleporterManager.LobbyTeleporterProp
            Character := Agent.GetFortCharacter[]
            Player := player[Agent]
            Player.IsActive[]
            #PlayerData:=GameManager.PlayersDataManager.GetPlayerData[Player]
        then: 
            GameManager.VFXManager.AudioToSingle_PlayerFail.Play(Agent)
            DistanceAmount:=Distance(Teleporter0.GetTransform().Translation, Character.GetTransform().Translation)
            NumSeconds:=DistanceAmount/AVGRunSpeed
            NumPenalizationHits:=NumSeconds*AVGHit

            MaybeBakery := GameManager.GetAgentBakery(Agent)
            if(ValidBakery := MaybeBakery?):
                var TotalDamage : float = GetPickaxeDamage(false,ValidBakery.PickaxeTotalStrength,false,ValidBakery.PickaxeCriticalProb,ValidBakery.PickaxeCriticalDamage,ValidBakery.PickaxeWeakPointMult,1.0)(0)
                set TotalDamage = TotalDamage + TotalDamage*HardBread_Tier_Multiplier*ValidBakery.PlayerDataCache.LV_Area*ValidBakery.Secret_PickaxeEfficiency
                if:
                    WaterEfficiency := ValidBakery.AllPickaxeEfficiency[WATER]
                    FlourEfficiency := ValidBakery.AllPickaxeEfficiency[FLOUR]
                    CheeseEfficiency := ValidBakery.AllPickaxeEfficiency[CHEESE]
                    PimentosEfficiency := ValidBakery.AllPickaxeEfficiency[PIMENTOS]
                then:
                    var TotalCoal_ToPenalize : float = 0.0
                    var TotalMutantMaterial_ToPenalize : float = CheckFinalMutantMaterialAward(ValidBakery.PlayerDataCache.LV_Area,NumPenalizationHits,1.0)
                    var TotalWater_ToPenalize : float = (TotalDamage*NumPenalizationHits*WaterEfficiency)/4.0
                    var TotalFlour_ToPenalize : float = (TotalDamage*NumPenalizationHits*FlourEfficiency)/4.0
                    var TotalCheese_ToPenalize : float = (TotalDamage*NumPenalizationHits*CheeseEfficiency)/4.0
                    var TotalPimentos_ToPenalize : float = (TotalDamage*NumPenalizationHits*PimentosEfficiency)/4.0

                    if(CoalToRemove:=Ceil[CoalProb*NumPenalizationHits]):
                        set TotalCoal_ToPenalize = CoalToRemove * 1.0
                    
                    if(WasOnDungeon?):
                        set TotalCoal_ToPenalize = TotalCoal_ToPenalize * 0.3
                        set TotalMutantMaterial_ToPenalize = TotalMutantMaterial_ToPenalize * 0.3
                        if(InCoal := Ceil[TotalCoal_ToPenalize]):
                            set TotalCoal_ToPenalize = InCoal*1.0
                        set TotalWater_ToPenalize = TotalWater_ToPenalize * 0.3
                        set TotalFlour_ToPenalize = TotalFlour_ToPenalize * 0.3
                        set TotalCheese_ToPenalize = TotalCheese_ToPenalize * 0.3
                        set TotalPimentos_ToPenalize = TotalPimentos_ToPenalize * 0.3

                    AllResourcersAmmount :=  array{TotalWater_ToPenalize*-1.0,TotalFlour_ToPenalize*-1.0,TotalCheese_ToPenalize*-1.0,TotalPimentos_ToPenalize*-1.0,TotalMutantMaterial_ToPenalize*-1.0,TotalCoal_ToPenalize*-1.0}
                    AllResourcersType := array{WATER,FLOUR,CHEESE,PIMENTOS,MUTANTMATERIAL,COAL}
                    AllLimitsReached := array{false,false,false,false,false,false}

                    var NewData : player_data_table = GameManager.PlayersDataManager.WithdrawResource(ValidBakery.PlayerDataCache,TotalWater_ToPenalize,WATER)
                    set NewData = GameManager.PlayersDataManager.WithdrawResource(NewData,TotalFlour_ToPenalize,FLOUR)
                    set NewData = GameManager.PlayersDataManager.WithdrawResource(NewData,TotalCheese_ToPenalize,CHEESE)
                    set NewData = GameManager.PlayersDataManager.WithdrawResource(NewData,TotalPimentos_ToPenalize,PIMENTOS)
                    set NewData = GameManager.PlayersDataManager.WithdrawResource(NewData,TotalCoal_ToPenalize,COAL)
                    set NewData = GameManager.PlayersDataManager.WithdrawResource(NewData,TotalMutantMaterial_ToPenalize,MUTANTMATERIAL)
                    #GameManager.PlayersDataManager.SaveNewData(Player,PlayerData,NewData)
                    ValidBakery.SetNewPlayerDataCache(NewData)
                    #GameManager.PlayersDataManager.SaveAllNewData(Player,PlayerData,NewData, ValidBakery)
                    ValidBakery.DisplayDamageUI(Agent, AllResourcersAmmount,AllResourcersType,AllLimitsReached,false)
                    ValidBakery.RefreshInventoryUI(NewData)
                    
    FindNearDungeon(Translation : vector3) : int =
        var NearstDungeon : int = 0
        for (Index:=0..AllDungeons_Positions.Length-1, Dungeon_Positions := AllDungeons_Positions[Index]):
            if:
                MiniBossSpawnPosition := Dungeon_Positions[2]
                2048.0>Distance(MiniBossSpawnPosition.GetTransform().Translation,Translation)
            then:
                ##Print("Found nearst dungeon {NearstDungeon}")
                set NearstDungeon = Index
        return NearstDungeon

    GetDamagePerHit(ObjetiveCharacterTransform : transform, Agent : agent) : tuple(float,int,logic) =

        #Do some basic math, it corresponds from character position?
        var TotalDamage : float = 0.0
        var CriticalProb : float = 0.0
        var CriticalMultiplier : float = 1.0
        var WeaponType : int = -99
        var IsCritical : logic = false

        if(Player := player[Agent], GameManager := MaybeGameManager?):
            MaybeBakery:=GameManager.GetAgentBakery(Agent)
            if(ValidBakery:=MaybeBakery?, WeaponsUpgradesUI:=ValidBakery.MaybeWeaponsUpgradesUI?, PetManager :=ValidBakery.PetsManager):

                BowExtraDamage_Multiplier := PetManager.PlayerEnhacement_BowExtraDamage
                SwordExtraDamage_Multiplier := PetManager.PlayerEnhacement_SwordExtraDamage
                SwordSuperStrike := PetManager.PlayerEnhacement_SwordSuperStrike
                WeaponsExtraCriticalDamage := PetManager.PlayerEnhacement_WeaponsExtraCritical

                #Log("Calcualting damage",LOG_LEVEL_DEBUG4)
                if(BowEquiped_ConditionalButton.IsHoldingItem[Agent]):
                    #Log("Arrow shoot detected",LOG_LEVEL_DEBUG4)
                    set CriticalMultiplier = WeaponsUpgradesUI.CriticalDamage_Bow
                    set TotalDamage = WeaponsUpgradesUI.Damage_Bow * BowExtraDamage_Multiplier
                    set WeaponType = BOW_TYPE 
                else if(SwordEquiped_ConditionalButton.IsHoldingItem[Agent]):
                    ShootDistance := Distance(WeaponsUpgradesUI.LastFiredPosition.Translation,ObjetiveCharacterTransform.Translation)
                    Speed:=((ShootDistance/100.0)/WeaponsUpgradesUI.LastTimeBowFired)

                    if(Speed>10.0,Speed<9999999999999.9):
                        Log("ShootDistance: {GetOptimizedMagnitudeLetter(ShootDistance)}, Speed: {GetOptimizedMagnitudeLetter(Speed)} Bow detected",LOG_LEVEL_DEBUG4)
                        ##Print("Arrow shoot detected for speed {Speed}, Distance {ShootDistance} on Time {WeaponsUpgradesUI.LastTimeBowFired}")
                        set CriticalMultiplier = WeaponsUpgradesUI.CriticalDamage_Bow
                        set TotalDamage = WeaponsUpgradesUI.Damage_Bow * BowExtraDamage_Multiplier
                        set WeaponType = BOW_TYPE 
                    else:
                        Log("ShootDistance: {GetOptimizedMagnitudeLetter(ShootDistance)}, Speed: {GetOptimizedMagnitudeLetter(Speed)} Sword detected",LOG_LEVEL_DEBUG4)
                        ##Print("Speed is too slow so it is a sword hit")
                        set CriticalMultiplier = WeaponsUpgradesUI.CriticalDamage_Sword
                        set TotalDamage = WeaponsUpgradesUI.Damage_Sword * SwordExtraDamage_Multiplier
                        TupleDamage := PetManager.CheckSwordStrikeEvent(GameManager.TotalGameTime,TotalDamage)
                        set TotalDamage = TupleDamage(0)
                        set WeaponType = SWORD_TYPE 
                        if(TupleDamage(1)?):
                            spawn:
                                GameManager.VFXManager.PlaySound(SUPERSTRIKE_SOUND,0.0,option{Agent},false)
                else:
                    ##Print("No damage")

                #Apply Stats
                CriticalRoll:=GetRandomFloat(0.0,0.1)
                if(CriticalRoll<CriticalProb):
                    set IsCritical = true
                    set TotalDamage = TotalDamage * CriticalMultiplier * WeaponsExtraCriticalDamage
                if(IsCritical?):
                    spawn:
                        GameManager.VFXManager.PlaySound(WEAPONCRITICAL_SOUND,0.0,option{Agent},false)

            else:
                PrintWarning("on GetDamagePerHit, invalid damage detected")

        return (TotalDamage,WeaponType,IsCritical)
    InitOrganizationArrays() : void =    

        #Chest Beacon
        set AllDungeons_ChestBeaconsInUse = array{AllDungeon0_ChestBeacon,AllDungeon1_ChestBeacon,AllDungeon2_ChestBeacon,
            AllDungeon4_ChestBeacon,AllDungeon5_ChestBeacon,AllDungeon6_ChestBeacon,AllDungeon7_ChestBeacon}

        #Mob info to load
        set AllDungeons_AllMobInfoToLoad = array{Dungeon0_AllMobInfoToLoad,Dungeon1_AllMobInfoToLoad,Dungeon2_AllMobInfoToLoad,
            Dungeon3_AllMobInfoToLoad,Dungeon4_AllMobInfoToLoad,Dungeon5_AllMobInfoToLoad,Dungeon6_AllMobInfoToLoad,
            Dungeon7_AllMobInfoToLoad}

        #MobBossinfo
        set AllDungeons_AllMobBossesInfoToLoad = array{Dungeon0_AllMobBossesInfoToLoad,Dungeon1_AllMobBossesInfoToLoad,Dungeon2_AllMobBossesInfoToLoad,
            Dungeon3_AllMobBossesInfoToLoad,Dungeon4_AllMobBossesInfoToLoad,Dungeon5_AllMobBossesInfoToLoad,Dungeon6_AllMobBossesInfoToLoad,
            Dungeon7_AllMobBossesInfoToLoad}

        #NPCSpawners
        set AllDungeons_NPCSpawnersInUse = array{AllDungeon0_NPCSpawnerUsed,AllDungeon1_NPCSpawnerUsed,AllDungeon2_NPCSpawnerUsed,
            AllDungeon3_NPCSpawnerUsed,AllDungeon4_NPCSpawnerUsed,AllDungeon5_NPCSpawnerUsed,AllDungeon6_NPCSpawnerUsed,
            AllDungeon7_NPCSpawnerUsed}

        #Positions
        set AllDungeons_Positions = array{AllDungeon0_Positions,AllDungeon1_Positions,AllDungeon2_Positions,AllDungeon3_Positions,
         AllDungeon4_Positions,AllDungeon5_Positions,AllDungeon6_Positions,AllDungeon7_Positions}

        #All agents playing
        set AllDungeons_AgentsPlaying = array{Dungeon0_AgentsPlaying,Dungeon1_AgentsPlaying,Dungeon2_AgentsPlaying,Dungeon3_AgentsPlaying,
            Dungeon4_AgentsPlaying,Dungeon5_AgentsPlaying,Dungeon6_AgentsPlaying,Dungeon7_AgentsPlaying}

        #All Dungeon MobsSpawned
        set AllDungeons_MobsSpawned = array{Dungeon0_MobsSpawned,Dungeon1_MobsSpawned,Dungeon3_MobsSpawned,Dungeon4_MobsSpawned,
            Dungeon5_MobsSpawned,Dungeon6_MobsSpawned,Dungeon7_MobsSpawned}

        #All Dungeon Mob Distribution
        set AllDungeons_MobsDistributionToBeSpawn = array{Dungeon0_MobsDistribution,Dungeon1_MobsDistribution,Dungeon2_MobsDistribution,
            Dungeon3_MobsDistribution,Dungeon4_MobsDistribution,Dungeon5_MobsDistribution,Dungeon6_MobsDistribution,
            Dungeon7_MobsDistribution}

        #All Dungeon Buttons
        set AllDungeons_Buttons = array{AllDungeon0_Buttons,AllDungeon1_Buttons,AllDungeon2_Buttons,AllDungeon3_Buttons,AllDungeon4_Buttons,
            AllDungeon5_Buttons,AllDungeon6_Buttons,AllDungeon7_Buttons}

        #All agents waiting
        set AllDungeons_AgentsWaitingToJoin = array{Dungeon0_AgentsWaiting,Dungeon1_AgentsWaiting,Dungeon2_AgentsWaiting,Dungeon3_AgentsWaiting,
            Dungeon4_AgentsWaiting,Dungeon5_AgentsWaiting,Dungeon6_AgentsWaiting,Dungeon7_AgentsWaiting}

        for(ChestBeacons : AllDungeonsChestBeacons):
            set AllDungeonsChestBeacons_IsUsed = AllDungeonsChestBeacons_IsUsed + array{false}

        #Is ready to play
        for(Dungeon : AllDungeons_Positions, DummyAgent := agent[GetPlayspace().GetPlayers()[0]]):
            set AllDungeons_IsReadyToPlay = AllDungeons_IsReadyToPlay + array{true} #Is ready to play
            set AllDungeons_TotalTime = AllDungeons_TotalTime + array{99} #Time
            set AllDungeons_HasStarted = AllDungeons_HasStarted + array{false}
            set AllDungeons_IsPublic = AllDungeons_IsPublic + array{false}
            set AllDungeons_MaybeStartingAgent = AllDungeons_MaybeStartingAgent + array{option{DummyAgent}}
            set AllDungeons_CountToStartTime = AllDungeons_CountToStartTime + array{10}
            set AllDungeons_MinibossIsDefeated = AllDungeons_MinibossIsDefeated + array{false}
            set AllDungeons_FinalMusicIsPlaying = AllDungeons_FinalMusicIsPlaying + array{false}
        for(Index:=0..AllDungeons_MaybeStartingAgent.Length-1, set AllDungeons_MaybeStartingAgent[Index]=false):

    InitHBHealthArray() : void =
        #HBLevels
        for(Index:=0..AllDungeons_Positions.Length-1, Dungeons_Positions := AllDungeons_Positions[Index], Prop := Dungeons_Positions[0]):
            HBLevel:=FindNearestHBLV(Prop)
            #Log("Dungeon: {Index}, HBLevel: {HBLevel}")
            set AllDungeons_HBAreaLevel = AllDungeons_HBAreaLevel + array{HBLevel}
        #Dungeon HitDamage and Mob Base Damage
        if(GameManager := MaybeGameManager?):
            AVGHits := (MinNum_AVGHit+MaxNum_AVGHit)/2.0
            for(DungeonIndex:=0..10):
                if:
                    HBLevel:=AllDungeons_HBAreaLevel[DungeonIndex]
                    ExpectedSapphires:=GameManager.PickaxeCalculator.AllLVs_ZoneCosts[HBLevel] #HBLevel = LevelArea
                then:
                    ExpectedMutantMaterial:=CheckFinalMutantMaterialAward(HBLevel,AVGHits,1.0)
                    TotalExpectedMutantMaterial:=ExpectedMutantMaterial*ReqDestroyedHB_PerPurity
                    Log("Dungeon: {DungeonIndex}, HBLevel: {HBLevel} Expected Sapphires: {GetOptimizedMagnitudeLetter(ExpectedSapphires)}, expected MutantMaterial per HB: {GetOptimizedMagnitudeLetter(ExpectedMutantMaterial)}, Total: {GetOptimizedMagnitudeLetter(TotalExpectedMutantMaterial)}",LOG_LEVEL_DEBUG7)
                    set AllDungeons_ExpectedHammerDamage = AllDungeons_ExpectedHammerDamage + array{GetExpectedWeaponDamageAVG(ExpectedSapphires,TotalExpectedMutantMaterial)}
                    set AllDungeons_MobBaseDamage = AllDungeons_MobBaseDamage + array{GetHitMobDamage(DungeonIndex)}
                
    CalculateMobDamage(MOB_TYPE : int, HBAreaLV : int, NumDungeon : int) : tuple(float,float)  =
        var MainDamage : float = 5.0
        var SecondDamage  : float = 5.0

        var Main_DamageMultiplier : float = 1.0
        var Second_DamageMultiplier : float = 1.0

        if(MOB_TYPE=MINIBOSS_MOB):
            set Main_DamageMultiplier = 1.5
            set Second_DamageMultiplier = 1.5
        else if(MOB_TYPE=PLANT_MOB):
            set Main_DamageMultiplier = 2.0
            set Second_DamageMultiplier = 2.0

        if(NumDungeon<>-1):
            if(Damage:=AllDungeons_MobBaseDamage[NumDungeon]):
                set MainDamage = Damage
                set SecondDamage = Damage
                Log("Equal Mob.NumDungeon: {NumDungeon}. MobBaseDamage: {Damage}", LOG_LEVEL_DEBUG9)
            else:
                PrintFatalError("On CalculateMobDamage: NumDungeon: {NumDungeon} Invalid!")
        else:
            var DungeonEqualFound : int = 0
            var IndexDungeon : int = 0
            loop:
                if(HBAreaLevel_OfDungeon := AllDungeons_HBAreaLevel[IndexDungeon], HBAreaLevel_OfDungeon<=HBAreaLV):
                    set DungeonEqualFound = IndexDungeon
                else:
                    break
                set IndexDungeon = IndexDungeon + 1
            
            if(Damage:=AllDungeons_MobBaseDamage[DungeonEqualFound]):
                set MainDamage = Damage
                set SecondDamage = Damage
                Log("Equal Mob.NumDungeon: {DungeonEqualFound}. MobBaseDamage: {Damage}", LOG_LEVEL_DEBUG9)
            else:
                PrintFatalError("On CalculateMobDamage: DungeonEqualFound: {DungeonEqualFound} Invalid!")

        return (MainDamage*Main_DamageMultiplier,SecondDamage*Second_DamageMultiplier)
         
    AddAgentToChestRegistration(Agent : agent) : void =
        if(set Dungeon_TresureRegistration[Agent] = array{}):
        
    OnPlayerAdded(Player : player) : void =
        if(Agent := agent[Player], Character := Agent.GetFortCharacter[]):
            #GranAgentWeapons(Agent)
            AddAgentToChestRegistration(Agent)
            #Damage_Zero.Pickup(Agent)
            Character.EliminatedEvent().Subscribe(OnPlayerEliminated)
        RefreshSwordEffects()

    OnPlayerRemoved(Player : player) : void =
        if(Agent := player[Player]):
            set Dungeon_TresureRegistration=RemoveAgentFromTreasureRegistrationMap(Dungeon_TresureRegistration,Agent)
            RemoveAgentFromAllDungeons(Agent)
            RefreshSwordEffects()
            

    RefreshSwordEffects() : void =
        set AllPlayersKnifeEffects = array{}
        AllPlayers:=GetPlayspace().GetPlayers()

        for(Index:=0..AllPlayers.Length-1, Player:=AllPlayers[Index], Agent := agent[Player]):
            if(GameManager:=MaybeGameManager?):
                MaybeBakery:=GameManager.GetAgentBakery(Agent)
                if:
                    ValidBakery := MaybeBakery?
                then:
                    set AllPlayersKnifeEffects = AllPlayersKnifeEffects + array{ValidBakery.PlayerDataCache.Tier_Sword-1}

    CheckIfArrowFired(Agent : agent, PressedTime : float) : void =
        if(GameManager:=MaybeGameManager?):
            MaybeBakery:=GameManager.GetAgentBakery(Agent)
            if:
                ValidBakery := MaybeBakery?
                BowEquiped_ConditionalButton.IsHoldingItem[Agent]
                Character :=  player[Agent].GetFortCharacter[]
                ValidBakery.MaybeWeaponsUpgradesUI
                WeaponsUpgradesUI:=ValidBakery.MaybeWeaponsUpgradesUI?
            then:

                WeaponsUpgradesUI.SetArrowRegistration(Character.GetTransform(), TotalGameTime, PressedTime)

                AmountArrows:=HowManyArros_ConditionalButton.GetItemCount(Agent,0)
                NewData := PlayersDataManager.SetAmmoArrow(ValidBakery.PlayerDataCache,AmountArrows)
                Log("Num arrows on inventory: {AmountArrows}",LOG_LEVEL_DEBUG7)
                ValidBakery.SetNewPlayerDataCache(NewData)
            #else:
            #    PrintFatalError("On CheckIfArrowFired")

    RemoveAgentFromAllDungeons(Agent : agent) : void =
        for(Index:=0..AllDungeons_Positions.Length-1):
            QuickAgentFromDungeon(Agent,Index)

    CheckIfIsPlaying(Agent : agent) : int =
        var DetectedNumDungeon : int = -1
        for(NumDungeon:=0..AllDungeons_AgentsPlaying.Length-1, AllAgentsPlaying:=AllDungeons_AgentsPlaying[NumDungeon], AllAgentsWaiting:=AllDungeons_AgentsWaitingToJoin[NumDungeon]):
            for(AgentWaiting : AllAgentsWaiting, AgentWaiting=Agent):
                set DetectedNumDungeon = NumDungeon
            for(AgentPlaying : AllAgentsPlaying, AgentPlaying=Agent):
                set DetectedNumDungeon = NumDungeon
        return DetectedNumDungeon

    IsAgentInsideDungeon(Agent : agent) : int = 
        var NumDungeon : int = -1
        for(AuxNumDungeon:=0..AllDungeons_AgentsPlaying.Length-1, AllAgentsPlaying:=AllDungeons_AgentsPlaying[NumDungeon], AllAgentsWaiting:=AllDungeons_AgentsWaitingToJoin[NumDungeon]):
            for(AgentPlaying : AllAgentsPlaying, AgentPlaying=Agent):
                set NumDungeon = AuxNumDungeon
        return NumDungeon
            
    RemoveAgentSpecificFromDungeon(Agent : agent, NumDungeon : int) : void =
        var NeedToCheckEnd : logic = false
        if(AllDungeonAgents :=  AllDungeons_AgentsPlaying[NumDungeon], AllAgentsWaiting:=AllDungeons_AgentsWaitingToJoin[NumDungeon]):
            for(IndexAgent:=0..AllDungeonAgents.Length-1, AgentPlaying := AllDungeonAgents[IndexAgent], AgentPlaying = Agent):
                if:
                    Aux_DungeonAgents := AllDungeonAgents.RemoveElement[IndexAgent]
                    set AllDungeons_AgentsPlaying[NumDungeon] = Aux_DungeonAgents
                    set NeedToCheckEnd = true
            for(IndexAgent:=0..AllAgentsWaiting.Length-1, AgentWaiting := AllAgentsWaiting[IndexAgent], AgentWaiting = Agent):
                if:
                    Aux_DungeonAgents := AllAgentsWaiting.RemoveElement[IndexAgent]
                    set AllDungeons_AgentsWaitingToJoin[NumDungeon] = Aux_DungeonAgents
                    set NeedToCheckEnd = true
        if(NeedToCheckEnd?):
            CheckDungeonEnd(NumDungeon)

    InitSubscribers() : void =
        InputTriggerFire.ReleasedEvent.Subscribe(CheckIfArrowFired)
        for(IndexDungeon:=0..AllDungeons_Buttons.Length-1, DungeonButtons :=  AllDungeons_Buttons[IndexDungeon]):
            for(IndexButton:=0..DungeonButtons.Length-1, DungeonButton := DungeonButtons[IndexButton]):
                DungeonButton.InteractedWithEvent.Subscribe(handler_weaponsandmobs{MaybeGeneralMobsManager:=option{Self}, DungeonNumber:=IndexDungeon, DungeonButtonType:=IndexButton}.HandlerDungeonButton)
    #Executed on MiniBoss SpawnedEvent

    <#
    MiniBossSpawnDetected(MobAgent : agent) : void =
        var Index:int=0
        if:
            MiniBossInfoToLoad:=AllMobBossesInfoToLoad[0]
            MiniBossInfoToLoad(1)<>-1
            MobCharacter := MobAgent.GetFortCharacter[] 
        then:
            loop:
                if:
                    MobManager := AllMobsManager[Index]
                    not MobManager.IsBeingUsed?
                then:
                    ##Print("Found valid MobManager on {Index}")
                    MobManager.SetUse(true)
                    MobManager.SetStats(MiniBossInfoToLoad(2),MiniBossInfoToLoad(3),MiniBossInfoToLoad(4),MiniBossInfoToLoad(5))
                    MobManager.SetAgentAndCharacter(MobAgent, MobCharacter)
                    MobManager.Init()
                    if:
                        NewArray := AllMobBossesInfoToLoad.RemoveElement[0]
                        set AllMobBossesInfoToLoad = NewArray
                    break
                else:
                    set Index = Index + 1
                if(Index>AllMobsManager.Length-1):
                    PrintFatalError("FATAL ERROR, MOB MANAGER NOT FOUND")
                    break
        else:
            PrintFatalError("On MobSpawnDetected, MobInfoToLoad is invalid")#>


    #Executed on SpawnedEvent
    RandomEncounterMobSpawnDetected(MobAgent : agent) : void =
        var Index:int=0
        if:
            RandomEncounterMobInfoToLoad:=AllRandomEncounterMobInfoToLoad[0]
            MobCharacter := MobAgent.GetFortCharacter[] 
        then:
            if(RandomEncounterMobInfoToLoad(2)=MINIBOSS_MOB):
                PrintFatalError("MINIBOSS_MOB spawn detected and not is a miniboss!")
            else:
                loop:
                    if:
                        MobManager := AllMobsManager[Index]
                        not MobManager.IsBeingUsed?
                    then:
                        ##Print("Found valid MobManager on {Index}, setting stats")
                        MobManager.SetUse(true)
                        MobManager.SetStats(RandomEncounterMobInfoToLoad(2),RandomEncounterMobInfoToLoad(3),-1<#RandomEncounterMobInfoToLoad(4)#>, RandomEncounterMobInfoToLoad(5), RandomEncounterMobInfoToLoad(6))
                        MobManager.SetAgentAndCharacter(MobAgent, MobCharacter)
                        MobManager.SetPlayerOwner(RandomEncounterMobInfoToLoad(7))
                        MobManager.Init()
                        if:
                            NewArray := AllRandomEncounterMobInfoToLoad.RemoveElement[0]
                            set AllRandomEncounterMobInfoToLoad = NewArray
                        break
                    else:
                        set Index = Index + 1
                    if(Index>AllMobsManager.Length-1):
                        PrintFatalError("FATAL ERROR, MOB MANAGER NOT FOUND")
                        break
        else:
            PrintFatalError("On MobSpawnDetected, MobInfoToLoad is invalid: ")

    FindDungeonOfSpawner(IndexSpawner : int, MobType : int) : int =
        var NumDungeon : int = 0

        loop:
            if(AllIndexesUsed:=AllDungeons_NPCSpawnersInUse[NumDungeon], UsedIndex:=AllIndexesUsed[MobType]):
                if(UsedIndex=IndexSpawner):
                    break
            set NumDungeon = NumDungeon + 1
            if(NumDungeon > AllDungeons_NPCSpawnersInUse.Length-1):
                PrintFatalError("FindDungeonOfSpawner: Dungeon of spawner not found")
                set NumDungeon = -1
                break

        return NumDungeon

    DungeonMobSpawnDetected(MobAgent : agent, IndexSpawner : int, MobType : int) : void =
        var Index:int=0
        NumDungeon:=FindDungeonOfSpawner(IndexSpawner, MobType)

        var AllMobInfoToLoad : []tuple(transform,logic,int,int,int,logic,int) = array{}

        if(MobType=MINIBOSS_MOB, NumDungeon<>-1, Info := AllDungeons_AllMobBossesInfoToLoad[NumDungeon]):
            set AllMobInfoToLoad = Info
        else if(Info := AllDungeons_AllMobInfoToLoad[NumDungeon]):
            set AllMobInfoToLoad = Info
        if:
            NumDungeon<>-1
        then:
            if:
                MobInfoToLoad:=AllMobInfoToLoad[0]
                MobInfoToLoad(1)<>-1
                MobCharacter := MobAgent.GetFortCharacter[] 
                HBAreaLV:=AllDungeons_HBAreaLevel[NumDungeon]
            then:
                loop:
                    if:
                        MobManager := AllMobsManager[Index]
                        not MobManager.IsBeingUsed?
                    then:
                        ##Print("Found valid MobManager on {Index}")
                        MobManager.SetUse(true)
                        MobManager.SetStats(MobInfoToLoad(2),MobInfoToLoad(3),MobInfoToLoad(4), MobInfoToLoad(5),HBAreaLV)
                        MobManager.SetAgentAndCharacter(MobAgent, MobCharacter)
                        MobManager.Init()
                        if:
                            NewArray := AllMobInfoToLoad.RemoveElement[0]
                            set AllMobInfoToLoad = NewArray
                        then:
                            if(MobType=MINIBOSS_MOB, set AllDungeons_AllMobBossesInfoToLoad[NumDungeon] = AllMobInfoToLoad):
                            else if(set AllDungeons_AllMobInfoToLoad[NumDungeon] = AllMobInfoToLoad):
                            else:
                                PrintFatalError("DungeonMobSpawnDetected: On deleting MobInfo of dungeon {NumDungeon}")
                        break
                    else:
                        set Index = Index + 1
                    if(Index>AllMobsManager.Length-1):
                        PrintFatalError("FATAL ERROR, MOB MANAGER NOT FOUND")
                        break
            else:
                PrintFatalError("On MobSpawnDetected, MobInfoToLoad is invalid: ")
            ###############
                <#
                if(MobInfoToLoad(2)=MINIBOSS_MOB):
                    PrintFatalError("MINIBOSS_MOB spawn detected and not is a miniboss!")
                else:
                    loop:
                        if:
                            MobManager := AllMobsManager[Index]
                            not MobManager.IsBeingUsed?
                        then:
                            ##Print("Found valid MobManager on {Index}")
                            MobManager.SetUse(true)
                            MobManager.SetStats(MobInfoToLoad(2),MobInfoToLoad(3),MobInfoToLoad(4), MobInfoToLoad(5))
                            MobManager.SetAgentAndCharacter(MobAgent, MobCharacter)
                            MobManager.Init()
                            if:
                                NewArray := AllMobInfoToLoad.RemoveElement[0]
                                set AllMobInfoToLoad = NewArray
                            break
                        else:
                            set Index = Index + 1
                        if(Index>AllMobsManager.Length-1):
                            PrintFatalError("FATAL ERROR, MOB MANAGER NOT FOUND")
                            break
            else:
                PrintFatalError("On MobSpawnDetected, MobInfoToLoad is invalid: ")#>

    TeleportPlayerToDungeon(Agent : agent, NumDungeon : int) : void =
        #Teleport all agents
        if:
            GameManager:=MaybeGameManager?
            AgentBakery:=GameManager.GetAgentBakery_Transacted(Agent)?
            AllDungeonPosition:=AllDungeons_Positions[NumDungeon]
            StartPositionProp:=AllDungeonPosition[1]
            TranslationProp:=StartPositionProp.GetTransform().Translation
            RotationProp:=StartPositionProp.GetTransform().Rotation
            Player:=player[Agent]
            Player.IsActive[]
        then:
            AgentBakery.CloseAllUI(?IsComplete := false)
            NewTranslation:vector3=vector3{X:=TranslationProp.X,Y:=TranslationProp.Y,Z:=TranslationProp.Z+UpVisualEmptyHideZ+64.0}
            NewSpawnPosition:=transform{Translation:=NewTranslation,Rotation:=RotationProp,Scale:=StartPositionProp.GetTransform().Scale}
            if(Character := Agent.GetFortCharacter[], Character.IsActive[]):  
                Character.Heal(99999.0)               
                if:
                    Character.TeleportTo[NewTranslation,RotationProp]
                then:
                    GameManager.VFXManager.StopAllMusics(Agent)
                    #Log("Player teleported to Z: {NewTranslation.Z}",LOG_LEVEL_NPCDEBUG)
                    GameManager.VFXManager.SpawnTeleportEffect(Agent,GameManager.GetAgentIndex(Agent),NewSpawnPosition)
                else:
                    TupleResult:=TeleportFailedAttempt(option{Character},false,false,NewSpawnPosition,256.0)
                    if(not TupleResult(0)?):
                        QuickAgentFromDungeon(Agent, NumDungeon)
                    else:
                        GameManager.VFXManager.StopAllMusics(Agent)
                        GameManager.VFXManager.SpawnTeleportEffect(Agent,GameManager.GetAgentIndex(Agent),TupleResult(1))
                GameManager.ChangeAgentClass_ToDungeonMode(Agent)
                MaybeBakery:=GameManager.GetAgentBakery(Agent)
                if(ValidBakery:=MaybeBakery?):
                    ValidBakery.DisplayDungeonInfo()
                else:
                    PrintFatalError("On TeleportPlayerToDungeon, Bakery not found")
                ##Print("Player teleported")
        else:
            PrintFatalError("On TeleportPlayerToDungeon")


    ExecuteDungeon_MobSpawn(NumDungeon :int, SpawnPosition : transform, IsBoss : logic, MobType : int, MobLevel : int , IsBossChild : logic)<suspends> : void =
        
        #set AllMobInfoToLoad = AllMobInfoToLoad + array{MobInfo} #Save the new info to be loaded
        ##Print("Preparing dungeon spawn")
        ##Print("-----------------------")
        if:
            AllNPCSpawnersInUse:=AllDungeons_NPCSpawnersInUse[NumDungeon]
            IndexSpawnerToUse:=AllNPCSpawnersInUse[MobType]
        then:
            MobInfo := (SpawnPosition,IsBoss,MobType,MobLevel,NumDungeon,IsBossChild,IndexSpawnerToUse)
            var NPCSpawner_ToUse : npc_spawner_device = npc_spawner_device{}
            if(MobType = GROUND_MOB, Spawner := AllNPC_GroundMobSpawner[IndexSpawnerToUse], set NPCSpawner_ToUse = Spawner <#Root := AllNPC_GroundMobSpawner_Root[IndexSpawnerToUse],#> <#set NPCSpawnerRoot_ToUse = Root#>):
                ##Print("GROUND_MOB Spawner used: {IndexSpawnerToUse}")
                #set AllMobInfoToLoad = AllMobInfoToLoad + array{MobInfo} #Save the new info to be loaded
            else if(MobType = DRUID_MOB, Spawner := AllNPC_DruidMobSpawner[IndexSpawnerToUse], set NPCSpawner_ToUse = Spawner <#Root := AllNPC_DruidMobSpawner_Root[IndexSpawnerToUse],#> <# set NPCSpawnerRoot_ToUse = Root#>):
                ##Print("DRUID_MOB Spawner used: {IndexSpawnerToUse}")
                #set AllMobInfoToLoad = AllMobInfoToLoad + array{MobInfo} #Save the new info to be loaded
            else if(MobType = PLANT_MOB, Spawner := AllNPC_PlantMobSpawner[IndexSpawnerToUse], set NPCSpawner_ToUse = Spawner <#Root := AllNPC_PlantMobSpawner_Root[IndexSpawnerToUse],#> <# set NPCSpawnerRoot_ToUse = Root#>):
                ##Print("PLANT_MOB Spawner used: {IndexSpawnerToUse}")
                #set AllMobInfoToLoad = AllMobInfoToLoad + array{MobInfo} #Save the new info to be loaded
            else if(MobType = MINIBOSS_MOB, Spawner := AllNPC_MiniBossMobSpawner[IndexSpawnerToUse], set NPCSpawner_ToUse = Spawner <#Root := AllNPC_DruidMobSpawner_Root[IndexSpawnerToUse],#> <# set NPCSpawnerRoot_ToUse = Root#>):
                ##Print("MINIBOSS_MOB Spawner used: {IndexSpawnerToUse}")
                #set AllMobBossesInfoToLoad = AllMobBossesInfoToLoad + array{MobInfo} #Save the new info to be loaded
            else:
                PrintFatalError("AllNPC_MiniBossMobSpawner {IndexSpawnerToUse} not found")
            #var NPCSpawnerRoot_ToUse : creative_prop = creative_prop{}
            var NewTranslation : vector3 = vector3{X:=SpawnPosition.Translation.X,Y:=SpawnPosition.Translation.Y,Z:=SpawnPosition.Translation.Z}
            if(not IsBossChild?):
                set NewTranslation=vector3{X:=SpawnPosition.Translation.X,Y:=SpawnPosition.Translation.Y,Z:=SpawnPosition.Translation.Z+UpVisualEmptyHideZ}
            NewSpawnPosition:=transform{Translation:=NewTranslation,Rotation:=SpawnPosition.Rotation,Scale:=SpawnPosition.Scale}
            if(NPCSpawner_ToUse.TeleportTo[NewSpawnPosition]):
                SpawnTuple:=SpawnProp(DirtSpawnEffect,NewSpawnPosition)
                MaybeProp:=SpawnTuple(0)
                spawn:
                    MobBehaviourHelper.PlaySound(NewSpawnPosition,SOUND_DIRT,-1.0)
                if(ValidProp:=MaybeProp?):
                    spawn:
                        DestroyEffect(ValidProp) 
                Sleep(0.04)
                if(MobType = MINIBOSS_MOB):
                    if:
                        AllMiniBossInfo:=AllDungeons_AllMobBossesInfoToLoad[NumDungeon]
                        set AllDungeons_AllMobBossesInfoToLoad[NumDungeon] = AllMiniBossInfo + array{MobInfo}
                else:
                    if:
                        AllMobInfo:=AllDungeons_AllMobInfoToLoad[NumDungeon]
                        set AllDungeons_AllMobInfoToLoad[NumDungeon] = AllMobInfo + array{MobInfo}
                NPCSpawner_ToUse.Spawn()         
                ##Print("Spawned mob")
            else:
                TupleResult:=TeleportFailedAttempt(false,option{NPCSpawner_ToUse},false,NewSpawnPosition,40.0)
                if(not TupleResult(0)?):
                    PrintFatalError("On ExecuteSpawn trying to teleport NPC Spawner to {NewTranslation.X},{NewTranslation.Y},{NewTranslation.Z}")
        else:
            PrintFatalError("AllDungeons_NPCSpawnersInUse[{NumDungeon}] or AllNPCSpawnersInUse[{MobType}] not valid!")
        
    #We calcule the new mob stats here and prepare the prespawn so the mob can load the stats right after the spawn
    ExecuteEncounter_PlantMobSpawn(Agent : agent, SpawnPosition : transform, IsBoss : logic, Level : int, HBAreaLV : int)<suspends> : void =
        NumDungeon := -1
        
        ##Print("Preparing spawn of Plant Mob LV {Level}")
        MaybeSpawner:=FindSpawnerToBeUse(Agent)
        if:
            NPCSpawner:=MaybeSpawner?
            Player := player[Agent]
        then:
            RandomEncounterMobInfoToLoad := (SpawnPosition,IsBoss,PLANT_MOB,Level,NumDungeon,false,HBAreaLV,option{Player})

            NewTranslation:vector3=vector3{X:=SpawnPosition.Translation.X,Y:=SpawnPosition.Translation.Y,Z:=SpawnPosition.Translation.Z+UpVisualEmptyHideZ}
            NewSpawnPosition:=transform{Translation:=NewTranslation,Rotation:=SpawnPosition.Rotation,Scale:=SpawnPosition.Scale}

            if(NPCSpawner.TeleportTo[NewSpawnPosition]):
                SpawnTuple:=SpawnProp(DirtSpawnEffect,NewSpawnPosition)
                MaybeProp:=SpawnTuple(0)
                spawn:
                    MobBehaviourHelper.PlaySound(NewSpawnPosition,SOUND_DIRT,-1.0)
                if(ValidProp:=MaybeProp?):
                    spawn:
                        DestroyEffect(ValidProp) 
                Sleep(0.04)
                set AllRandomEncounterMobInfoToLoad = AllRandomEncounterMobInfoToLoad + array{RandomEncounterMobInfoToLoad}
                NPCSpawner.Spawn()
            else:
                PrintFatalError("On ExecuteSpawn trying to teleport NPC Spawner to {NewTranslation.X},{NewTranslation.Y},{NewTranslation.Z}")

    FindSpawnerToBeUse(Agent : agent) : ?npc_spawner_device =
        var MaybeSpawner : ?npc_spawner_device = false
        if(GameManager:=MaybeGameManager?):
            PlayerIndex:=GameManager.GetAgentIndex(Agent)
            if(NPC_PlantMobSpawner:=AllNPC_PlantMobSpawner_RandomEncounter[PlayerIndex]):
                set MaybeSpawner = option{NPC_PlantMobSpawner}
        return MaybeSpawner

    DeployDungeonMobMultiSpawn(NumDungeon : int, PositionProp : creative_prop, MobsToSpawn : int, IsBossChild : logic)<suspends> : void =
        AllEnemiesToSpawn:=DungeonDataBase.GetEnemyWave(MobsToSpawn)
        Position:=PositionProp.GetTransform()
        if:
            MaybeStartingAgent:=AllDungeons_MaybeStartingAgent[NumDungeon]
            StartingAgent:=MaybeStartingAgent?
            StartingPlayer:=player[StartingAgent]
            GameManager:=MaybeGameManager?
        then:
            MaybeBakery:=GameManager.GetAgentBakery(StartingAgent)
            if:
                ValidBakery := MaybeBakery?
                #Evolution:=ValidBakery.PlayerDataCache.AllEvolvedDungeons[NumDungeon]
            then:
                #Print("A miniboss is going to be spawned")
                TupleMobLevel:=DungeonDataBase.GetMobsLV(<#Evolution+#>NumDungeon)
                TupleMobsLV:=DungeonDataBase.GetMobsLV(NumDungeon)

                #MobLevel := DungeonDataBase.GetDungeonMobLevel(NumDungeon)

                var IsSpawnAllowed : logic = false

                var Num_Child_Spawned : int = 0
                var Num_NoChild_Spawned : int = 0
                var AllChildMobsSpawned : []mob = array{}
                var AllNoChildMobsSpawned : []mob = array{}
                for(Mob : AllMobsManager, Mob.MobType<>MINIBOSS_MOB, not Mob.DeathRunning?, Mob.IsBeingUsed?, Mob.NumDungeon=NumDungeon):
                    if(Mob.IsBossChild?):
                        set Num_Child_Spawned = Num_Child_Spawned + 1
                        set AllChildMobsSpawned = AllChildMobsSpawned + array{Mob}
                    else:
                        set Num_NoChild_Spawned = Num_NoChild_Spawned + 1
                        set AllNoChildMobsSpawned = AllNoChildMobsSpawned + array{Mob}

                if(Num_Child_Spawned>15 or Num_NoChild_Spawned>15):
                    #Spawn not allowed
                    set IsSpawnAllowed = false
                else if(Num_Child_Spawned+Num_NoChild_Spawned>14):
                    set IsSpawnAllowed = false
                else:
                    set IsSpawnAllowed = true
                if(IsSpawnAllowed?):
                    for(Index:=0..AllEnemiesToSpawn.Length-1, EnemyType := AllEnemiesToSpawn[Index]):
                        ExecuteDungeon_MobSpawn(NumDungeon,Position,false,EnemyType,TupleMobsLV(1),IsBossChild)
                        Sleep(0.25)

    DestroyEffect(CreativeProp : creative_prop)<suspends> : void =
        ##Print("Destroying effect spawned")
        Sleep(5.0)
        CreativeProp.Dispose()

    

    ResetChestRegistration(Agent : agent) : void =
        if:
            set Dungeon_TresureRegistration[Agent] = array{}
        #then:
            ##Print("Treasure Registration reseted")

    SpawnDungeonTime(NumDungeon : int)<suspends> : void =
        MaxTimeToPlay:=DungeonDataBase.GetDungeonMaxTime(NumDungeon)
        if(set AllDungeons_HasStarted[NumDungeon]=true):
        FindBeaconChests(NumDungeon)
        loop:
            if:
                TotalTime := AllDungeons_TotalTime[NumDungeon] - 1
                NumAgentsPlaying := AllDungeons_AgentsPlaying[NumDungeon].Length
                set AllDungeons_TotalTime[NumDungeon] = TotalTime
                GameManager := MaybeGameManager?
            then:
                #Action UI
                if(AllAgents := AllDungeons_AgentsPlaying[NumDungeon]):
                    for(Agent : AllAgents):
                        MaybeAgentBakery:=GameManager.GetAgentBakery(Agent)
                        if(ValidBakery:=MaybeAgentBakery?, ActionInfoUI := ValidBakery.MaybeActionInfoUI?):
                            if(AllTreasuresOpened:=Dungeon_TresureRegistration[Agent].Length,AllTreasures:=AllDungeons_Buttons[NumDungeon].Length-2):
                                ActionInfoUI.SetDungeon_TimeInfo("Limit time: {TotalTime}")
                                ActionInfoUI.SetDungeon_ChestInfo("Chests opened: {AllTreasuresOpened}/{AllTreasures}")
                                if(TotalTime=30 or TotalTime=20 or TotalTime=10):
                                    spawn:
                                        GameManager.VFXManager.PlaySound(TIMEENDING_SOUND,0.0,option{Agent},false)
                                if(TotalTime<30):
                                    ActionInfoUI.SetDungeon_TimeEndingInfo("Limit time: ",TotalTime)
                            else:
                                PrintFatalError("On SpawnDungeonTime, treasure not found")
                        else:
                            PrintFatalError("On SpawnDungeonTime, Bakery not found")
                else:
                    PrintFatalError("On SpawnDungeonTime")
                #Checking for Mob Spawn
                CheckDungeonMobSpawn(NumDungeon)
                #Dungeon UI
                if:
                    AllButtons := AllDungeons_Buttons[NumDungeon]
                    EnterButton := AllButtons[0]
                then:
                    #Dungeon UI
                    for(Player : GetPlayspace().GetPlayers(), Agent := player[Player]):
                        MaybeBakery:=GameManager.GetAgentBakery(Agent)
                        if:
                            ValidBakery:=MaybeBakery?
                            DungeonUI:=ValidBakery.MaybeDungeonUI?
                            DungeonUI.Opened_NumDungeon=NumDungeon
                        then:
                            DungeonUI.UpdateUI()
                    #BUTTON JOIN
                    if(IsPublic := AllDungeons_IsPublic[NumDungeon]):
                        RemainingTime:=DungeonDataBase.CalculeDungeon_RemainingTimeToJoin(NumDungeon,TotalTime)
                        if:
                            RemainingTime<=0
                            set AllDungeons_IsReadyToPlay[NumDungeon] = false    
                        then:
                            ##Print("Set to no ready because RemainingTime: {RemainingTime}")
                            EnterButton.SetInteractionText(BasicMessage("There is somebody playing this dungeon!"))
                        else:
                            if(IsPublic?):
                                EnterButton.SetInteractionText(BasicMessage("Join and help! {RemainingTime}"))
                            else:
                                EnterButton.SetInteractionText(BasicMessage("Private game. Wait until the current player finish"))

                if(NumAgentsPlaying<=0, set AllDungeons_IsReadyToPlay[NumDungeon] = true):
                    if(AllButtons := AllDungeons_Buttons[NumDungeon], EnterButton := AllButtons[0]):
                        EnterButton.SetInteractionText(BasicMessage("Clean dungeon"))
                    break
                if(TotalTime<=0):
                    TimeEnded(NumDungeon)
            else:
                PrintFatalError("On SpawnDungeonTime: if failed")
                break
            Sleep(1.0)

    CheckBossMusic(NumDungeon : int) : void =
        if:
            IsBeingPlayed:=AllDungeons_FinalMusicIsPlaying[NumDungeon]
            not IsBeingPlayed?
            set AllDungeons_FinalMusicIsPlaying[NumDungeon] = true
            AllAgents:=AllDungeons_AgentsPlaying[NumDungeon]
            GameManager := MaybeGameManager?
            
        then:
            for(Agent : AllAgents, AgentBakery := GameManager.GetAgentBakery_Transacted(Agent)?):
                #AgentBakery.
                AgentBakery.CheckMusicPlay(DUNGEONMINIBOSSMUSIC_SOUND)
        #else:
        #    PrintFatalError("On CheckBossMusic")

    CheckDungeonMobSpawn(NumDungeon : int) : void =
        ##Print("Checking to spawn mobs on {NumDungeon} A")
        if:
            AllPositions:=AllDungeons_Positions[NumDungeon]
            AllAgentsPlaying:=AllDungeons_AgentsPlaying[NumDungeon]
            var AllMobsSpawned : []int =AllDungeons_MobsSpawned[NumDungeon]
            AllMobsDistributionToBeSpawn:=AllDungeons_MobsDistributionToBeSpawn[NumDungeon]
        then:
            for(Index:=2..AllPositions.Length-1):
                if:
                    PositionProp := AllPositions[Index]
                    NumMobsToBeSpawned := AllMobsDistributionToBeSpawn[Index]
                    MobsSpawned := AllMobsSpawned[Index]
                    MobsSpawned < NumMobsToBeSpawned
                then:
                    var IsThereCharacter_Nearby : logic = false
                    PositionTranslation:=PositionProp.GetTransform().Translation
                    PositionTranslationToCalcule:=vector3{X:=PositionTranslation.X,Y:=PositionTranslation.Y,Z:=PositionTranslation.Z+UpVisualEmptyHideZ}
                    for(AgentPlaying : AllAgentsPlaying, CharacterPlaying := AgentPlaying.GetFortCharacter[]):
                        DistanceFromPoint := Distance(CharacterPlaying.GetTransform().Translation,PositionTranslationToCalcule)
                        if(DistanceFromPoint<DungeonDataBase.DungeonMobDistanceDetection):
                            set IsThereCharacter_Nearby = true
                            #Print("Player nearby at {DistanceFromPoint}, trying to spawn {NumMobsToBeSpawned} Mobs")
                    if:
                        IsThereCharacter_Nearby?
                        set AllMobsSpawned[Index]=NumMobsToBeSpawned
                        set AllDungeons_MobsSpawned[NumDungeon]=AllMobsSpawned
                    then:
                        spawn:
                            DeployDungeonMobMultiSpawn(NumDungeon,PositionProp,NumMobsToBeSpawned,false)

    CheckNPCSpawners() : void =
        #GROUND SPAWNER
        for(Index:=0..AllNPC_GroundMobSpawner_IsBeingUsed.Length-1):
            if:
                GroundMobSpawner_IsBeingUsed:=AllNPC_GroundMobSpawner_IsBeingUsed[Index]
                NPCMobSpawner:=AllNPC_GroundMobSpawner[Index]
                GroundMobSpawner_IsBeingUsed <> -1
                (GroundMobSpawner_IsBeingUsed + MAX_NPCMOBSPAWNER_TIMEUSE)*1.0 < TotalGameTime
            then:
                PrintWarning("CheckNPCSpawners: MAX NPC TIME USE REACHED!")
                if:
                    set AllNPC_GroundMobSpawner_IsBeingUsed[Index] = -1
                then:
                    NPCMobSpawner.DespawnAll(false)
                    #NPCMobSpawner.Disable()
        #DRUID SPAWNER
        for(Index:=0..AllNPC_DruidMobSpawner_IsBeingUsed.Length-1):
            if:
                DruidMobSpawner_IsBeingUsed:=AllNPC_DruidMobSpawner_IsBeingUsed[Index]
                NPCMobSpawner:=AllNPC_DruidMobSpawner[Index]
                DruidMobSpawner_IsBeingUsed <> -1
                (DruidMobSpawner_IsBeingUsed + MAX_NPCMOBSPAWNER_TIMEUSE)*1.0 < TotalGameTime
            then:
                PrintWarning("CheckNPCSpawners: MAX NPC TIME USE REACHED!")
                if:
                    set AllNPC_DruidMobSpawner_IsBeingUsed[Index] = -1
                then:
                    NPCMobSpawner.DespawnAll(false)

        #PLAT SPAWNER
        for(Index:=0..AllNPC_PlantMobSpawner_IsBeingUsed.Length-1):
            if:
                PlantMobSpawner_IsBeingUsed:=AllNPC_PlantMobSpawner_IsBeingUsed[Index]
                NPCMobSpawner:=AllNPC_PlantMobSpawner[Index]
                PlantMobSpawner_IsBeingUsed <> -1
                (PlantMobSpawner_IsBeingUsed + MAX_NPCMOBSPAWNER_TIMEUSE)*1.0 < TotalGameTime
            then:
                PrintWarning("CheckNPCSpawners: MAX NPC TIME USE REACHED!")
                if:
                    set AllNPC_PlantMobSpawner_IsBeingUsed[Index] = -1
                then:
                    NPCMobSpawner.DespawnAll(false)
                    #NPCMobSpawner.Disable()
        #MINIBOSS SPAWNER
        for(Index:=0..AllNPC_MiniBossMobSpawner_IsBeingUsed.Length-1):
            if:
                MiniBossMobSpawner_IsBeingUsed:=AllNPC_MiniBossMobSpawner_IsBeingUsed[Index]
                NPCMobSpawner:=AllNPC_MiniBossMobSpawner[Index]
                MiniBossMobSpawner_IsBeingUsed <> -1
                (MiniBossMobSpawner_IsBeingUsed + MAX_NPCMOBSPAWNER_TIMEUSE)*1.0 < TotalGameTime
            then:
                PrintWarning("CheckNPCSpawners: MAX NPC TIME USE REACHED!")
                if:
                    set AllNPC_MiniBossMobSpawner_IsBeingUsed[Index] = -1
                then:
                    NPCMobSpawner.DespawnAll(false)
                    #NPCMobSpawner.Disable()
                    
    FindBeaconChests(NumDungeon : int) : void =
        var AllBeaconIndexesToUse : []int = array{} 
        if(AllButtons:=AllDungeons_Buttons[NumDungeon], NumOfChests := AllButtons.Length-2):
            #Ground mob spawner
            var Index_Beacon : int = 0
            #Print("Num of chest to find: {NumOfChests}")
            loop:
                if:
                    IsBeingUse:=AllDungeonsChestBeacons_IsUsed[Index_Beacon]
                    not IsBeingUse?
                then:
                    set AllBeaconIndexesToUse = AllBeaconIndexesToUse + array{Index_Beacon}
                    #Print("Adding valid beacon: {Index_Beacon}")
                set Index_Beacon = Index_Beacon + 1
                if(AllBeaconIndexesToUse.Length>=NumOfChests):
                    #Print("Breaking because AllBeaconIndexesToUse.Length is {AllBeaconIndexesToUse.Length} and NumOfChests is {NumOfChests}")
                    break
                    
                if(Index_Beacon>AllDungeonsChestBeacons.Length-1):
                    PrintFatalError("AllDungeonsChestBeacons_IsUsed overflow on FindBeaconChests")
                    set AllBeaconIndexesToUse = array{}
                    break

            #Print("AllBeaconIndexesToUse.Length is {AllBeaconIndexesToUse.Length}")
            
            for(NumChest:=0..NumOfChests-1):
                if:
                    BeaconIndex:=AllBeaconIndexesToUse[NumChest]
                    BeaconDevice := AllDungeonsChestBeacons[BeaconIndex]
                    Button := AllButtons[NumChest+2]
                then:
                    #Print("It will be use the beacon {BeaconIndex} on the button {NumChest+2}")
                    ButtonTransform:=Button.GetTransform()
                    BeaconDevice.RemoveAllFromShowList()
                    if(BeaconDevice.TeleportTo[ButtonTransform], AllAgentsPlaying := AllDungeons_AgentsPlaying[NumDungeon]):
                        for(Agent : AllAgentsPlaying):
                            #Print("Added Agent to Becacon device")
                            BeaconDevice.AddToShowList(Agent)
        if(set AllDungeons_ChestBeaconsInUse[NumDungeon]=AllBeaconIndexesToUse):

    FindNearestHBLV(Prop : creative_prop) : int =
        Position := Prop.GetTransform().Translation
        var NearestDistance : float = 9999.9
        var NearestHBLV : int = 0
        if(GameManager:=MaybeGameManager?):
            for(HB : GameManager.AllHardBreads): #NumOfTypes
                DistanceBetween:=Distance(HB.SlateBreadProp.GetTransform().Translation,Position)
                if(DistanceBetween<NearestDistance):
                    set NearestHBLV = HB.Level
                    set NearestDistance = DistanceBetween
        return NearestHBLV

    FindDungeonMobSpawners(NumDungeon : int) : void =
        #Ground mob spawner
        var Index_GroundMobSpawner : int = 0
        loop:
            if(TimeWasBeingUse:=AllNPC_GroundMobSpawner_IsBeingUsed[Index_GroundMobSpawner], TimeWasBeingUse=-1):
                break
            else:
                set Index_GroundMobSpawner = Index_GroundMobSpawner + 1
            if(Index_GroundMobSpawner>AllNPC_GroundMobSpawner_IsBeingUsed.Length-1):
                PrintFatalError("GroundMobSpawner to be use not found!")
                set Index_GroundMobSpawner = -1
                break
        #Druid mob spawner
        var Index_DruidMobSpawner : int = 0
        loop:
            if(TimeWasBeingUse:=AllNPC_DruidMobSpawner_IsBeingUsed[Index_DruidMobSpawner], TimeWasBeingUse=-1):
                break
            else:
                set Index_DruidMobSpawner = Index_DruidMobSpawner + 1
            if(Index_DruidMobSpawner>AllNPC_DruidMobSpawner_IsBeingUsed.Length-1):
                PrintFatalError("DruidMobSpawner to be use not found!")
                set Index_DruidMobSpawner = -1
                break
        #Plantmob spawner
        var Index_PlantMobSpawner : int = 0
        loop:
            if(TimeWasBeingUse:=AllNPC_PlantMobSpawner_IsBeingUsed[Index_PlantMobSpawner], TimeWasBeingUse=-1):
                break
            else:
                set Index_PlantMobSpawner = Index_PlantMobSpawner + 1
            if(Index_PlantMobSpawner>AllNPC_PlantMobSpawner_IsBeingUsed.Length-1):
                PrintFatalError("PlantMobSpawner to be use not found!")
                set Index_PlantMobSpawner = -1
                break
        #MiniBoss mob spawner
        var Index_MiniBossMobSpawner : int = 0
        loop:
            if(TimeWasBeingUse:=AllNPC_MiniBossMobSpawner_IsBeingUsed[Index_MiniBossMobSpawner], TimeWasBeingUse=-1):
                break
            else:
                set Index_MiniBossMobSpawner = Index_MiniBossMobSpawner + 1
            if(Index_MiniBossMobSpawner>AllNPC_MiniBossMobSpawner_IsBeingUsed.Length-1):
                PrintFatalError("MiniBossMobSpawner to be use not found!")
                set Index_MiniBossMobSpawner = -1
                break

        #if(Index_GroundMobSpawner=-1 or Index_DruidMobSpawner=-1 or Index_MiniBossMobSpawner=-1):
        #    PrintFatalError("On FindDungeonMobSpawners no found. Invalid. Index_GroundMobSpawner: {Index_GroundMobSpawner}. Index_DruidMobSpawner: {Index_DruidMobSpawner}. Index_MiniBossMobSpawner: {Index_MiniBossMobSpawner}")
        if:
            set AllDungeons_NPCSpawnersInUse[NumDungeon] = array{Index_DruidMobSpawner,Index_GroundMobSpawner,Index_PlantMobSpawner,Index_MiniBossMobSpawner}
            TimeToInt:=Floor[TotalGameTime]
            set AllNPC_DruidMobSpawner_IsBeingUsed[Index_DruidMobSpawner] = TimeToInt
            set AllNPC_GroundMobSpawner_IsBeingUsed[Index_GroundMobSpawner] = TimeToInt
            set AllNPC_PlantMobSpawner_IsBeingUsed[Index_PlantMobSpawner] = TimeToInt
            set AllNPC_MiniBossMobSpawner_IsBeingUsed[Index_MiniBossMobSpawner] = TimeToInt
            DruidMobSpawner:=AllNPC_DruidMobSpawner[Index_DruidMobSpawner]
            GroundMobSpawner:=AllNPC_GroundMobSpawner[Index_GroundMobSpawner]
            PlantMobSpawner:=AllNPC_PlantMobSpawner[Index_PlantMobSpawner]
            MiniBossMobSpawner:=AllNPC_MiniBossMobSpawner[Index_MiniBossMobSpawner]
        then:
            #GroundMobSpawner.Enable()
            #DruidMobSpawner.Enable()
            #MiniBossMobSpawner.Enable()
            DruidMobSpawner.DespawnAll(false)
            GroundMobSpawner.DespawnAll(false)
            PlantMobSpawner.DespawnAll(false)
            MiniBossMobSpawner.DespawnAll(false)
        else:
            PrintFatalError("NPC Spawners not found!")
        



    TimeEnded(NumDungeon : int) : void =
        #Print("The time to play the dungeon has ended!")
        if(GameManager := MaybeGameManager?, AllAgentsPlaying:=AllDungeons_AgentsPlaying[NumDungeon]):
            for(AgentPlaying : AllAgentsPlaying):
                GameManager.DisplayCharacterMessage(AgentPlaying, G_CHARACTER,"Time ended!")
                spawn:
                    GameManager.VFXManager.PlaySound(LOSE_SOUND,0.0,option{AgentPlaying},false)
                QuickAgentFromDungeon(AgentPlaying,NumDungeon)

    QuickAgentFromDungeon(Agent : agent, NumDungeon : int) : void =
        ##Print("Trying to abandon dungeon {NumDungeon}")
        RemoveAllChestBeacons(Agent,NumDungeon)
        ResetChestRegistration(Agent)
        if(GameManager := MaybeGameManager?<#, MainInventoryU:=TestsManager.AllMainInventoryUI[Agent], #>, Player:= player[Agent],  Player.IsActive[], AgentBakery := GameManager.GetAgentBakery_Transacted(Agent)?, HardBreadLevel:=AllDungeons_HBAreaLevel[NumDungeon]):
            if(HardBreadLevel>13):
                AgentBakery.CheckMusicPlay(SECRETMUSIC_SOUND)
            else if(HardBreadLevel>7):
                AgentBakery.CheckMusicPlay(DESERTMUSIC_SOUND)
            else:
                AgentBakery.CheckMusicPlay(BAKERSMUSIC_SOUND)
        
            #AgentBakery.CheckMusicPlay(DUNGEONMINIBOSSMUSIC_SOUND)
            MaybeBakery:=GameManager.GetAgentBakery(Agent)
            if(ValidBakery := MaybeBakery?):
                ValidBakery.RemoveDungeonInfo()
                GameManager.ChangeAgentClass_ToPickMode(Agent)
        RemoveAgentSpecificFromDungeon(Agent,NumDungeon)
        if(AllPositions:=AllDungeons_Positions[NumDungeon], ExitPosition:=AllPositions[0]):
            RotationExit:=ExitPosition.GetTransform().Rotation
            NewTranslation:vector3=vector3{X:=ExitPosition.GetTransform().Translation.X,Y:=ExitPosition.GetTransform().Translation.Y,Z:=ExitPosition.GetTransform().Translation.Z+UpVisualEmptyHideZ+64.0}
            NewExitPosition:=transform{Translation:=NewTranslation,Rotation:=RotationExit,Scale:=ExitPosition.GetTransform().Scale}
            if(Character:=Agent.GetFortCharacter[], GameManager := MaybeGameManager?):
                if(Player:= player[Agent],  Player.IsActive[], Character.IsActive[]):
                    Character.SetHealth(Character.GetMaxHealth())   
                    if(Character.TeleportTo[NewTranslation,RotationExit]):
                        #Log("Player teleported to Z: {NewTranslation.Z}",LOG_LEVEL_NPCDEBUG)
                        GameManager.VFXManager.SpawnTeleportEffect(Agent,GameManager.GetAgentIndex(Agent),NewExitPosition)
                    else:
                        TupleResult:=TeleportFailedAttempt(option{Character},false,false,NewExitPosition,256.0)
                        if(not TupleResult(0)?):
                            Character.Damage(5000.0)
                        else:
                            GameManager.VFXManager.SpawnTeleportEffect(Agent,GameManager.GetAgentIndex(Agent),TupleResult(1))
                            TupleResult(1)
        else:
            PrintFatalError("Position {NumDungeon} not found")

    CheckDungeonEnd(NumDungeon : int) : void =
        if(GameManager := MaybeGameManager?,AllPlayers:=AllDungeons_AgentsPlaying[NumDungeon],AllPlayerWaiting:=AllDungeons_AgentsWaitingToJoin[NumDungeon]):
            if(AllPlayers.Length<=0, AllPlayerWaiting.Length<=0):
                if(set AllDungeons_HasStarted[NumDungeon]=false):
                ResetDungeon(NumDungeon)
                if(AllButtons:=AllDungeons_Buttons[NumDungeon],EntryButton:=AllButtons[0]):
                    EntryButton.SetInteractionText(BasicMessage("Clean dungeon"))
                    GameManager.UpdateAll_DungeonUI(NumDungeon)
            else:
                #Print("Not ending because AllPlayers.Length: {AllPlayers.Length} and AllPlayerWaiting.Length: {AllPlayerWaiting.Length}")

    MobSpawnerCleaner(NumDungeon : int)<suspends> : void =
        #Detect all mobs
        var AllMobsSpawned : []mob = array{}
        for(Mob : AllMobsManager, not Mob.DeathRunning?, Mob.IsBeingUsed?, Mob.NumDungeon=NumDungeon):
            set AllMobsSpawned = AllMobsSpawned + array{Mob}
        if:
            NPCSpawnersInUse:=AllDungeons_NPCSpawnersInUse[NumDungeon]
            GroundSpawnerInUse := AllNPC_GroundMobSpawner[NPCSpawnersInUse[DRUID_MOB]]
            DruidSpawnerInUse := AllNPC_DruidMobSpawner[NPCSpawnersInUse[GROUND_MOB]]
            PlantSpawnerInUse := AllNPC_PlantMobSpawner[NPCSpawnersInUse[PLANT_MOB]]
            MiniBossSpawnerInUse := AllNPC_MiniBossMobSpawner[NPCSpawnersInUse[MINIBOSS_MOB]]
        then:
            #Kill mobs
            for(Index:=0..5, Mob := AllMobsSpawned[Index]):
                Sleep(0.1)
                Mob.DespawnMobV2(true)
            Sleep(0.3)
            for(Index:=5..10, Mob := AllMobsSpawned[Index]):
                Sleep(0.1)
                Mob.DespawnMobV2(true)
            Sleep(0.3)
            for(Index:=10..15, Mob := AllMobsSpawned[Index]):
                Sleep(0.1)
                Mob.DespawnMobV2(true)
            Sleep(0.3)
            for(Index:=15..20, Mob := AllMobsSpawned[Index]):
                Sleep(0.1)
                Mob.DespawnMobV2(true)
            Sleep(0.5)
            #Clean register
            if:
                set AllNPC_GroundMobSpawner_IsBeingUsed[NPCSpawnersInUse[DRUID_MOB]] = -1
                set AllNPC_DruidMobSpawner_IsBeingUsed[NPCSpawnersInUse[GROUND_MOB]] = -1
                set AllNPC_PlantMobSpawner_IsBeingUsed[NPCSpawnersInUse[PLANT_MOB]] = -1
                set AllNPC_MiniBossMobSpawner_IsBeingUsed[NPCSpawnersInUse[MINIBOSS_MOB]] = -1
            then:
                GroundSpawnerInUse.DespawnAll(false)
                DruidSpawnerInUse.DespawnAll(false)
                PlantSpawnerInUse.DespawnAll(false)
                MiniBossSpawnerInUse.DespawnAll(false)    
        if:
            GameManager:=MaybeGameManager?
            AllButtons := AllDungeons_Buttons[NumDungeon]
            EntryButton := AllButtons[0]
            set AllDungeons_NPCSpawnersInUse[NumDungeon] = array{-1,-1,-1,-1}
            set AllDungeons_MobsSpawned[NumDungeon] = array{}
            set AllDungeons_MinibossIsDefeated[NumDungeon] = false
            set AllDungeons_IsReadyToPlay[NumDungeon] = true
        then:
            if(AgentsWaiting:=AllDungeons_AgentsWaitingToJoin[NumDungeon]):
                #Print("Reset dungeon completed, agents waiting: {AgentsWaiting.Length}")
            GameManager.UpdateAll_DungeonUI(NumDungeon)
            EntryButton.SetInteractionText(BasicMessage("Play dungeon"))

    ResetDungeon(NumDungeon : int) : void =        
        #Delete all currect NPCS, despawn and disable
        spawn:
            MobSpawnerCleaner(NumDungeon)
        if:
            ChestBeaconsInUse := AllDungeons_ChestBeaconsInUse[NumDungeon]
            set AllDungeons_FinalMusicIsPlaying[NumDungeon] = false
            set AllDungeons_HasStarted[NumDungeon]=false
        then:
            #ChestBeacons
            for (Index : ChestBeaconsInUse):
                if:
                    set AllDungeonsChestBeacons_IsUsed[Index] = false
                    Beacon := AllDungeonsChestBeacons[Index]
                then:
                    Beacon.RemoveAllFromShowList()
                    BeaconPosition:=Beacon.GetTransform().Translation
                    NewBeaconPosition:=vector3{X:=BeaconPosition.X,Y:=BeaconPosition.Y,Z:=BeaconPosition.Z-1024.0}
                    if(Beacon.TeleportTo[NewBeaconPosition,Beacon.GetTransform().Rotation]):
        if:
            GameManager:=MaybeGameManager?
            AllButtons := AllDungeons_Buttons[NumDungeon]
            EntryButton := AllButtons[0]
            set AllDungeons_AllMobInfoToLoad[NumDungeon] = array{}
            set AllDungeons_AllMobBossesInfoToLoad[NumDungeon] = array{}
            set AllDungeons_ChestBeaconsInUse[NumDungeon] = array{}
            set AllDungeons_MinibossIsDefeated[NumDungeon] = false
            set AllDungeons_CountToStartTime[NumDungeon] = 10
            set AllDungeons_IsPublic[NumDungeon] = false
            set AllDungeons_MaybeStartingAgent[NumDungeon] = false
            set AllDungeons_AgentsWaitingToJoin[NumDungeon] = array{}
            set AllDungeons_TotalTime[NumDungeon] = 1
            set AllDungeons_AgentsPlaying[NumDungeon] = array{}
        then:
            if(AgentsWaiting:=AllDungeons_AgentsWaitingToJoin[NumDungeon]):
                #Print("Reset dungeon completed, agents waiting: {AgentsWaiting.Length}")
            GameManager.UpdateAll_DungeonUI(NumDungeon)
            EntryButton.SetInteractionText(BasicMessage("Loading..."))
        else:
            PrintFatalError("On ResetDungeon")
    <#
    ResetDungeon(NumDungeon : int) : void =        
        #Delete all currect NPCS, despawn and disable
        if:
            ChestBeaconsInUse := AllDungeons_ChestBeaconsInUse[NumDungeon]
            NPCSpawnersInUse:=AllDungeons_NPCSpawnersInUse[NumDungeon]
            GroundSpawnerInUse := AllNPC_GroundMobSpawner[NPCSpawnersInUse[DRUID_MOB]]
            DruidSpawnerInUse := AllNPC_DruidMobSpawner[NPCSpawnersInUse[GROUND_MOB]]
            PlantSpawnerInUse := AllNPC_PlantMobSpawner[NPCSpawnersInUse[PLANT_MOB]]
            MiniBossSpawnerInUse := AllNPC_MiniBossMobSpawner[NPCSpawnersInUse[MINIBOSS_MOB]]
            set AllNPC_GroundMobSpawner_IsBeingUsed[NPCSpawnersInUse[DRUID_MOB]] = -1
            set AllNPC_DruidMobSpawner_IsBeingUsed[NPCSpawnersInUse[GROUND_MOB]] = -1
            set AllNPC_PlantMobSpawner_IsBeingUsed[NPCSpawnersInUse[PLANT_MOB]] = -1
            set AllNPC_MiniBossMobSpawner_IsBeingUsed[NPCSpawnersInUse[MINIBOSS_MOB]] = -1
            set AllDungeons_FinalMusicIsPlaying[NumDungeon] = false
            set AllDungeons_HasStarted[NumDungeon]=false
        then:
            #Print("Dungeon AllDungeons_HasStarted [{NumDungeon}]is set to false")
            GroundSpawnerInUse.DespawnAll(false)
            DruidSpawnerInUse.DespawnAll(false)
            PlantSpawnerInUse.DespawnAll(false)
            MiniBossSpawnerInUse.DespawnAll(false)
            #ChestBeacons
            for (Index : ChestBeaconsInUse):
                if:
                    set AllDungeonsChestBeacons_IsUsed[Index] = false
                    Beacon := AllDungeonsChestBeacons[Index]
                then:
                    Beacon.RemoveAllFromShowList()
                    BeaconPosition:=Beacon.GetTransform().Translation
                    NewBeaconPosition:=vector3{X:=BeaconPosition.X,Y:=BeaconPosition.Y,Z:=BeaconPosition.Z-1024.0}
                    if(Beacon.TeleportTo[NewBeaconPosition,Beacon.GetTransform().Rotation]):
            #GroundSpawnerInUse.Disable()
            #DruidSpawnerInUse.Disable()
            #MiniBossSpawnerInUse.Disable()
        if:
            TestsManager:=MaybeTestsManager?
            AllButtons := AllDungeons_Buttons[NumDungeon]
            EntryButton := AllButtons[0]
            set AllDungeons_ChestBeaconsInUse[NumDungeon] = array{}
            set AllDungeons_NPCSpawnersInUse[NumDungeon] = array{-1,-1,-1,-1}
            set AllDungeons_MobsSpawned[NumDungeon] = array{}
            set AllDungeons_MinibossIsDefeated[NumDungeon] = false
            set AllDungeons_CountToStartTime[NumDungeon] = 10
            set AllDungeons_IsPublic[NumDungeon] = false
            set AllDungeons_MaybeStartingAgent[NumDungeon] = false
            set AllDungeons_AgentsWaitingToJoin[NumDungeon] = array{}
            set AllDungeons_IsReadyToPlay[NumDungeon] = true
            set AllDungeons_TotalTime[NumDungeon] = 1
            set AllDungeons_AgentsPlaying[NumDungeon] = array{}
        then:
            if(AgentsWaiting:=AllDungeons_AgentsWaitingToJoin[NumDungeon]):
                #Print("Reset dungeon completed, agents waiting: {AgentsWaiting.Length}")
            TestsManager.UpdateAll_DungeonUI(NumDungeon)
            EntryButton.SetInteractionText(BasicMessage("Play dungeon"))#>

    OnMiniBossDeath(NumDungeon : int) : void =
        #Print("MiniBossDungeon defeated")
        if:
            set AllDungeons_MinibossIsDefeated[NumDungeon] = true
            AllAgentsPlaying:=AllDungeons_AgentsPlaying[NumDungeon]
            GameManager := MaybeGameManager?
        then:
            for(Agent : AllAgentsPlaying):
                GameManager.VFXManager.StopAllMusics(Agent)
                spawn:
                    GameManager.VFXManager.PlaySound(WIN_SOUND,0.0,option{Agent},false)

    #Start dungeon
    StartDungeon(Agent : agent, NumDungeon : int): void =
        #Print("Starting Dungeon: {NumDungeon}")
        if(GameManager := MaybeGameManager?, AgentBakery := GameManager.GetAgentBakery_Transacted(Agent)? <#, MainInventoryU:=TestsManager.AllMainInventoryUI[Agent]#>):
            AgentBakery.CheckMusicPlay(DUNGEONMUSIC_SOUND)
        #Other check, that the starting player is available
        if(GameManager := MaybeGameManager?, MaybeStartingAgent:=AllDungeons_MaybeStartingAgent[NumDungeon], ValidStartingAgent := MaybeStartingAgent?):
            MaxTimeToPlay:=DungeonDataBase.GetDungeonMaxTime(NumDungeon)
            if:
                AllAgentsWaitingToJoin := AllDungeons_AgentsWaitingToJoin[NumDungeon]
                set AllDungeons_AgentsPlaying[NumDungeon] = array{} #Clear memory
                set AllDungeons_AgentsPlaying[NumDungeon] = AllAgentsWaitingToJoin #Set the new agents
                set AllDungeons_TotalTime[NumDungeon] = MaxTimeToPlay #TotaTime
                NumPositions := AllDungeons_Positions[NumDungeon]
            then:
                var ArrayMobsSpawned : []int = array{}
                for(Index:=0..NumPositions.Length-1):
                    set ArrayMobsSpawned = ArrayMobsSpawned + array{0}
                MobsDistribution := DungeonDataBase.GetNumMobsToBeSpawn(NumDungeon,NumPositions.Length)
                #Consume one try to starting player and prepare mob spawns
                MaybeBakery:=GameManager.GetAgentBakery(Agent)
                if:
                    BakeryManager:=MaybeBakery?
                    Player:=player[Agent]
                    Player.IsActive[]
                    BakeryManager.PlayerDataCache.Dungeon_Accumulated>0
                    set AllDungeons_MobsDistributionToBeSpawn[NumDungeon] = MobsDistribution
                    set AllDungeons_MobsSpawned[NumDungeon] = ArrayMobsSpawned    
                then:
                    NewPlayerData:=GameManager.PlayersDataManager.ConsumeDungeonAccumulated(BakeryManager.PlayerDataCache)
                    BakeryManager.DisplayResourcersGather(array{-1.0},array{DUNGEONTOKEN},false)
                    #BakeryManager.GameManager.VFXManager.PlayerPays_AudioPlayer.Play(Agent)
                    BakeryManager.SetNewPlayerDataCache(NewPlayerData)
                    #GameManager.PlayersDataManager.SaveAllNewData(Player,PlayerData,NewPlayerData, BakeryManager)
                    #GameManager.PlayersDataManager.SaveNewData(Player,PlayerData,NewPlayerData)
                    FindDungeonMobSpawners(NumDungeon)
                    #Spawn MiniBoss
                    if:
                        AllDungeonPosition := AllDungeons_Positions[NumDungeon]
                        BossPosition := AllDungeonPosition[2]
                        set AllDungeons_MinibossIsDefeated[NumDungeon] = false
                        Evolution:=BakeryManager.PlayerDataCache.AllEvolvedDungeons[NumDungeon]
                    then:
                        #Print("A miniboss is going to be spawned")
                        TupleMobLevel:=DungeonDataBase.GetMobsLV(Evolution+NumDungeon)
                        spawn:
                            ExecuteDungeon_MobSpawn(NumDungeon,BossPosition.GetTransform(),true,MINIBOSS_MOB,TupleMobLevel(0),false)
                    else:
                        PrintFatalError("AllDungeonPosition{NumDungeon} or BossPosition{2} not found")
                    #Start dungeon timer
                    spawn:
                        SpawnDungeonTime(NumDungeon)
                    #Teleport all agents
                    for(AgentToTeleport : AllAgentsWaitingToJoin):     
                        TeleportPlayerToDungeon(AgentToTeleport,NumDungeon) 
                    Delay_DungeonMusic(AllAgentsWaitingToJoin) 
            else:
                PrintFatalError("StartDungeon has failed!")   

    CheckInPlayerCanJoin(Agent : agent, NumDungeon : int) : void =
        if(GameManager:=MaybeGameManager?):
            var IsAlreadyInOtherDungeon : logic = false
            for(Index:=0..AllDungeons_AgentsWaitingToJoin.Length-1):
                if: 
                    Index <> NumDungeon
                    AllAgentsWaiting := AllDungeons_AgentsWaitingToJoin[Index]
                    AllAgentsPlaying := AllDungeons_AgentsPlaying[Index]
                then:
                    for(OtherAgent : AllAgentsWaiting, OtherAgent = Agent):
                        set IsAlreadyInOtherDungeon = true
                    for(OtherAgent : AllAgentsPlaying, OtherAgent = Agent):
                        set IsAlreadyInOtherDungeon = true
            MaybeBakery:=GameManager.GetAgentBakery(Agent)  
            if(not IsAlreadyInOtherDungeon?,ValidBakery:=MaybeBakery?):  
                ValidBakery.Display_DungeonUI(Agent,NumDungeon)
            else:
                PrintWarning("You are in other dungeon joined")
                
    Delay_DungeonMusic(AllAgentsWaitingToJoin : []agent) : void =
        if(GameManager:=MaybeGameManager?):
            for(AgentToTeleport : AllAgentsWaitingToJoin, AgentBakery := GameManager.GetAgentBakery_Transacted(AgentToTeleport)?):
                spawn:
                    DelayMusic(AgentBakery)     

    DelayMusic(AgentBakery : bakery_machines_manager)<suspends> : void =
        Sleep(1.0)
        AgentBakery.CheckMusicPlay(DUNGEONMUSIC_SOUND)


    TryJoinPlayerTo_InitiatedDungeon(Agent : agent, NumDungeon : int) : void =
        #Game has no started, join to the waiting room
        var IsNotAlreadyIn : logic = false
        if:
            var AllAgentsPlaying : []agent = AllDungeons_AgentsPlaying[NumDungeon]
            var AllAgentsWaitingToJoin : []agent = AllDungeons_AgentsWaitingToJoin[NumDungeon]
            HasStarted := AllDungeons_HasStarted[NumDungeon]
            IsPublic := AllDungeons_IsPublic[NumDungeon]
            TotalTime := AllDungeons_TotalTime[NumDungeon]
        then:
            for(JoinedAgent : AllAgentsWaitingToJoin, JoinedAgent = Agent):
                set IsNotAlreadyIn = true
            for(JoinedAgent : AllAgentsWaitingToJoin, JoinedAgent = Agent):
                set IsNotAlreadyIn = true
            if(not IsNotAlreadyIn?):
                if(IsPublic?):
                    if(HasStarted?):
                        #Has started, can join?
                        RemainingTime:=DungeonDataBase.CalculeDungeon_RemainingTimeToJoin(NumDungeon,TotalTime)
                        if:
                            RemainingTime>0
                        then:
                            set AllAgentsPlaying = AllAgentsPlaying + array{Agent}
                            set AllAgentsWaitingToJoin = AllAgentsWaitingToJoin + array{Agent}
                            if(set AllDungeons_AgentsPlaying[NumDungeon] = AllAgentsPlaying, set AllDungeons_AgentsWaitingToJoin[NumDungeon] = AllAgentsWaitingToJoin):
                                TeleportPlayerToDungeon(Agent,NumDungeon)
                        else:
                            #Print("Is too late to join!")
                    else:
                        #Has no started, so join him to the waiting room
                        set AllAgentsPlaying = AllAgentsPlaying + array{Agent}
                        set AllAgentsWaitingToJoin = AllAgentsWaitingToJoin + array{Agent}
                        if(set AllDungeons_AgentsPlaying[NumDungeon] = AllAgentsPlaying, set AllDungeons_AgentsWaitingToJoin[NumDungeon] = AllAgentsWaitingToJoin):

    DungeonUI_PlayButtonPressed(Agent : agent, NumDungeon : int) : void =
        if(GameManager := MaybeGameManager?):
            MaybeBakery:=GameManager.GetAgentBakery(Agent)
            if(ValidBakery:=MaybeBakery?, DungeonUI := ValidBakery.MaybeDungeonUI?):
                if(IsReady := AllDungeons_IsReadyToPlay[NumDungeon], IsReady?):
                    if:
                        var AllAgentsPlaying : []agent = AllDungeons_AgentsPlaying[NumDungeon]
                        var AllAgentsWaitingToPlay : []agent = AllDungeons_AgentsWaitingToJoin[NumDungeon]
                    then:
                        if(AllAgentsPlaying.Length<=0, AllAgentsWaitingToPlay.Length<=0): 
                            #There is no playes playing or waiting, trying to start a game
                            #Start if only has permisions acumulated
                            if:
                                Player:=player[Agent]
                                Player.IsActive[]
                                ValidBakery.PlayerDataCache.Dungeon_Accumulated>0
                            then:
                                #P#Print("Runing countdown")
                                spawn:
                                    AddToDungeonAndStartCountDown(Agent,NumDungeon)
                                DungeonUI.UpdateUI()
                            else:
                                spawn:
                                    DungeonUI.RunNoCleaningPermit()
                        else if(MaybeAgent:=AllDungeons_MaybeStartingAgent[NumDungeon],ValidStartingAgent:=MaybeAgent?, ValidStartingAgent=Agent):
                            PrintWarning("TODO Ask for help on DungeonUI_PlayButtonPressed")
                            if(set AllDungeons_IsPublic[NumDungeon]=true):
                                GameManager.UpdateAll_DungeonUI(NumDungeon)
                                #P#Print("Now is public")
                            DungeonUI.UpdateUI()
                        else if(HasStarted:=AllDungeons_HasStarted[NumDungeon], IsPublic := AllDungeons_IsPublic[NumDungeon], IsPublic?):
                            if(not HasStarted?):
                                #P#Print("The game has not started") 
                                TryJoinPlayerTo_InitiatedDungeon(Agent,NumDungeon)
                            else:
                                ##Print("Trying to join a started dungeon")
                                #Check if the player can join
                                MaxTimeToPlay:=GameManager.GeneralMobsManager.DungeonDataBase.GetDungeonMaxTime(NumDungeon)
                                if:
                                    TotalTime:=AllDungeons_TotalTime[NumDungeon]
                                    RemaingTimeToJoin:=TotalTime-(MaxTimeToPlay-(DungeonDataBase.MaxTimeToEnter))
                                then:
                                    #P#Print("RemaingTimeToJoin:  {RemaingTimeToJoin}")
                                    if(RemaingTimeToJoin>0):
                                        var IsNotAlreadyIn : logic = false
                                        for(JoinedAgent : AllAgentsPlaying, JoinedAgent = Agent):
                                            set IsNotAlreadyIn = true
                                        for(JoinedAgent : AllAgentsWaitingToPlay, JoinedAgent = Agent):
                                            set IsNotAlreadyIn = true
                                        if(not IsNotAlreadyIn?):
                                            if(IsPublic?):
                                                if(HasStarted?):
                                                    #Has started, can join?
                                                    RemainingTime:=DungeonDataBase.CalculeDungeon_RemainingTimeToJoin(NumDungeon,TotalTime)
                                                    if:
                                                        RemainingTime>0
                                                    then:
                                                        set AllAgentsPlaying = AllAgentsPlaying + array{Agent}
                                                        set AllAgentsWaitingToPlay = AllAgentsWaitingToPlay + array{Agent}
                                                        if(set AllDungeons_AgentsPlaying[NumDungeon] = AllAgentsPlaying, set AllDungeons_AgentsWaitingToJoin[NumDungeon] = AllAgentsWaitingToPlay):
                                                            TeleportPlayerToDungeon(Agent,NumDungeon)
                                                    else:
                                                        #Print("Is too late to join!")
                                                else:
                                                    #Has no started, so join him to the waiting room
                                                    set AllAgentsPlaying = AllAgentsPlaying + array{Agent}
                                                    set AllAgentsWaitingToPlay = AllAgentsWaitingToPlay + array{Agent}
                                                    if(set AllDungeons_AgentsPlaying[NumDungeon] = AllAgentsPlaying, set AllDungeons_AgentsWaitingToJoin[NumDungeon] = AllAgentsWaitingToPlay):         
                                        <#
                                            if:
                                                AllDungeonPosition:=AllDungeons_Positions[NumDungeon]
                                                StartPositionProp:=AllDungeonPosition[1]
                                                TranslationProp:=StartPositionProp.GetTransform().Translation
                                                RotationProp:=StartPositionProp.GetTransform().Rotation
                                            then:
                                                if(Character := Agent.GetFortCharacter[]):                 
                                                    if:
                                                        Character.TeleportTo[TranslationProp,RotationProp]
                                                    else:
                                                        PrintWarning("TODO, safe teleport to other place if it fails")
                                            else:
                                                PrintFatalError("Fatal error tring to join player")#>
                                    else:
                                        #Print("IsToo Late!")   
                            DungeonUI.UpdateUI()  

                else:
                    PrintFatalError("Fatal error, is not ready!")  
            else:
                PrintFatalError("Fatal error, trying to find tests manager or Dungeon UI")                          
            

    DungeonUI_CancelButtonPressed(Agent : agent, NumDungeon : int) : void =
        if(GameManager := MaybeGameManager?):
            MaybeBakery := GameManager.GetAgentBakery(Agent)
            if(ValidBakery := MaybeBakery?, DungeonUI := ValidBakery.MaybeDungeonUI?, ActionUI := ValidBakery.MaybeActionInfoUI?):
                RemoveAgentSpecificFromDungeon(Agent,NumDungeon)     
                ActionUI.RemoveDungeonInfo()

            #ungeonUI.UpdateUI()
    RemoveAllChestBeacons(Agent : agent, NumDungeon : int) : void =
        if(AllChestsIndexBeacons:=AllDungeons_ChestBeaconsInUse[NumDungeon]):
            for(IndexBeaconToRemove : AllChestsIndexBeacons):
                if(Beacon := AllDungeonsChestBeacons[IndexBeaconToRemove]):
                    Beacon.RemoveFromShowList(Agent)
                else:
                    PrintFatalError("Beacon {IndexBeaconToRemove} not found") 
        else:
            PrintFatalError("Dungeon {NumDungeon} not found")

    AddToDungeonAndStartCountDown(Agent : agent, NumDungeon : int)<suspends> : void =
        if(GameManager := MaybeGameManager?):
            MaybeBakery := GameManager.GetAgentBakery(Agent)
            if(ValidBakery := MaybeBakery?, DungeonUI := ValidBakery.MaybeDungeonUI?):
                if(var AllAgents : []agent = AllDungeons_AgentsPlaying[NumDungeon]):
                    if(AllAgents.Length<=0): #THERE IS NO PLAYERS, LETS PLAY IT
                        set AllAgents = AllAgents + array{Agent} #Add agent
                        if:
                            set AllDungeons_MaybeStartingAgent[NumDungeon]=option{Agent}
                            set AllDungeons_AgentsWaitingToJoin[NumDungeon]=AllAgents
                            set AllDungeons_CountToStartTime[NumDungeon] = 10
                            var TimeToJoin:  int = AllDungeons_CountToStartTime[NumDungeon]
                        then:
                            loop:
                                for(AgentWaiting:AllAgents):
                                    OtherMaybeBakery := GameManager.GetAgentBakery(Agent)
                                    if(ValidOtherBakery := OtherMaybeBakery?, ActionUI := ValidOtherBakery.MaybeActionInfoUI?):
                                        ActionUI.SetDungeon_TimeInfo("Starting in: {TimeToJoin}")
                                        ActionUI.SetDungeon_ChestInfo("Find the chests and destroy the mutant bread")
                                ##Print("Countdown: {TimeToJoin}")
                                GameManager.UpdateAll_DungeonUI(NumDungeon)
                                Sleep(1.0)
                                set TimeToJoin = TimeToJoin - 1
                                if(set AllDungeons_CountToStartTime[NumDungeon]=TimeToJoin):
                                if(TimeToJoin<=0, MaybeStartingAgent:=AllDungeons_MaybeStartingAgent[NumDungeon], ValidStartingAgent := MaybeStartingAgent?):
                                    StartDungeon(ValidStartingAgent, NumDungeon)
                                    break
                                if(MaybeStartingAgent:=AllDungeons_MaybeStartingAgent[NumDungeon], ValidStartingAgent := MaybeStartingAgent?):
                                else:
                                    break

    DungeonButton_Pressed(Agent : agent, NumDungeon : int, ButtonType : int)<suspends> : void =
        if(GameManager:=MaybeGameManager?):
            if(ButtonType=ENTER_TYPE_BUTTON):
                CheckInPlayerCanJoin(Agent,NumDungeon)
            else if(ButtonType=EXIT_TYPE_BUTTON):
                spawn:
                    GameManager.VFXManager.PlaySound(LOSE_SOUND, 0.0, option{Agent}, false)
                QuickAgentFromDungeon(Agent,NumDungeon)
            else:
                IsAlreadyOpened:=TryOpenChest(Agent,NumDungeon,ButtonType)
                if(not IsAlreadyOpened?, HBLevelArea := AllDungeons_HBAreaLevel[NumDungeon]):
                    if(Player := player[Agent]):
                        if(ButtonType=2):
                            if:
                                IsDefeated := AllDungeons_MinibossIsDefeated[NumDungeon]
                                IsDefeated?
                            then:
                                spawn:
                                    GameManager.VFXManager.PlaySound(CHEST_SOUND, 0.0, option{Agent}, false)
                                AllAwardsTuple:=DungeonDataBase.GetAndSaveChestAwards(Player,ButtonType,NumDungeon,HBLevelArea)
                                MaybeBakery:=GameManager.GetAgentBakery(Agent)
                                if(ValidBakery:=MaybeBakery?, MainInventoryUI := ValidBakery.MaybeMainInventoryUI?, ActionInfoUI := ValidBakery.MaybeActionInfoUI?):
                                    ValidBakery.DisplayResourcersGather(AllAwardsTuple(0),AllAwardsTuple(1),false)
                                    ValidBakery.RefreshInventoryUI(Agent)
                                if(AllTreasuresOpened:=Dungeon_TresureRegistration[Agent].Length,AllTreasures:=AllDungeons_Buttons[NumDungeon].Length-2, AllTreasuresOpened=AllTreasures):
                                    GameManager.DisplayCharacterMessage(Agent,G_CHARACTER, "Good job cleaning the dungeon!")
                                    spawn:
                                        GameManager.VFXManager.PlaySound(WIN_SOUND,0.0,option{Agent},false)
                                    QuickAgentFromDungeon(Agent,NumDungeon)
                                    if(ValidBakery:=MaybeBakery?, MainInventoryUI := ValidBakery.MaybeMainInventoryUI?, Player.IsActive[]):
                                        PlayerDataCache := GameManager.QuestsAndOrdersDatabase.AddAgentQuest_ProgressV2(Player,ValidBakery.PlayerDataCache,1.0,COMPLETEDUNGEON_TUTORIAL_TYPE)(4)
                                        ValidBakery.SetNewPlayerDataCache(PlayerDataCache)
                                        GameManager.DisplayCharacterMessage(Agent,G_CHARACTER, "Good job!")
                                    
                            else:
                                #PrintWarning("TODO: Warn you must to defeat the miniboss before")
                                GameManager.DisplayCharacterMessage(Agent,I_CHARACTER,"You must defeat the miniboss before open this chest!", ?Sound := ERROR_SOUND)
                        else:       
                            spawn:
                                GameManager.VFXManager.PlaySound(CHEST_SOUND,0.0,option{Agent},false)
                            AllAwardsTuple:=DungeonDataBase.GetAndSaveChestAwards(Player,ButtonType,NumDungeon,HBLevelArea)
                            MaybeBakery:=GameManager.GetAgentBakery(Agent)
                            if(ValidBakery:=MaybeBakery?):
                                ValidBakery.RefreshInventoryUI(Player)
                                ValidBakery.DisplayResourcersGather(AllAwardsTuple(0),AllAwardsTuple(1),false)
                                if(Player.IsActive[], AllTreasuresOpened:=Dungeon_TresureRegistration[Agent].Length,AllTreasures:=AllDungeons_Buttons[NumDungeon].Length-2, AllTreasuresOpened=AllTreasures):
                                    #Print("All treasures completed")
                                    PlayerDataCache := GameManager.QuestsAndOrdersDatabase.AddAgentQuest_ProgressV2(Player,ValidBakery.PlayerDataCache,1.0,COMPLETEDUNGEON_TUTORIAL_TYPE)(4)
                                    ValidBakery.SetNewPlayerDataCache(PlayerDataCache)
                                    GameManager.DisplayCharacterMessage(Agent, G_CHARACTER, "Good job cleaning the dungeon!")
                                    spawn:
                                        GameManager.VFXManager.PlaySound(WIN_SOUND,0.0,option{Agent},false)
                                    QuickAgentFromDungeon(Agent,NumDungeon)
                                    GameManager.DisplayCharacterMessage(Agent, G_CHARACTER, "Good job cleaning the dungeon!")
                else:
                    spawn:
                        GameManager.VFXManager.PlaySound(ERROR_SOUND,0.0,option{Agent},false)
                    GameManager.DisplayCharacterMessage(Agent, I_CHARACTER, "You have already opened this chest", ?Sound := ERROR_SOUND)
                    

                #PrintWarning("TO DO Dungeon reward get from pickaxe calculator")
        else:
            PrintFatalError("TestsManager not found")
        
    TryOpenChest(Agent : agent, NumDungeon : int, NumChest : int) : logic =
        var IsAlreadyOpened : logic = false 
        if(var TreasureRegistration : []int = Dungeon_TresureRegistration[Agent]):
            for(TreasureOpened : TreasureRegistration, TreasureOpened = NumChest):
                set IsAlreadyOpened = true
            if(IsAlreadyOpened?):
            else:
                if(NumChest=2):
                    if:
                        IsDefeated := AllDungeons_MinibossIsDefeated[NumDungeon]
                        IsDefeated?
                    then:
                        set TreasureRegistration = TreasureRegistration + array{NumChest}
                        if(set Dungeon_TresureRegistration[Agent] = TreasureRegistration):
                            set IsAlreadyOpened = false
                            RegisterChestBeacon(Agent,NumDungeon,NumChest)
                else:
                    set TreasureRegistration = TreasureRegistration + array{NumChest}
                    if(set Dungeon_TresureRegistration[Agent] = TreasureRegistration):
                        set IsAlreadyOpened = false
                        RegisterChestBeacon(Agent,NumDungeon,NumChest)
        else:
            PrintFatalError("Treasure registration of this agent does not exists")
        return IsAlreadyOpened

    RegisterChestBeacon(Agent : agent, NumDungeon : int, NumChest : int) : void =
        #Print("Trying to remove to do list of NumDungeon: {NumDungeon} beacon chest {NumChest}")

        if(AllChestsIndexBeacons:=AllDungeons_ChestBeaconsInUse[NumDungeon]):
            if(IndexBeaconToRemove := AllChestsIndexBeacons[NumChest-2]):
                if(Beacon := AllDungeonsChestBeacons[IndexBeaconToRemove]):
                    Beacon.RemoveFromShowList(Agent)
                else:
                    PrintFatalError("Beacon {IndexBeaconToRemove} not found") 
            else:
                PrintFatalError("AllChestsIndexBeacons {NumChest-2} not found")
        else:
            PrintFatalError("Dungeon {NumDungeon} not found")
    CreateMobsManagers() : void =
        for(Index:=0..100):
            NewMob := MakeMobManagerUI(Self, Index)
            set AllMobsManager = AllMobsManager + array{NewMob}
            ##Print("New mob manager set on {Index}")

    InitNPCSpawners() : void =
        for(Index:=0..AllNPC_PlantMobSpawner_RandomEncounter.Length-1, Spawner := AllNPC_PlantMobSpawner_RandomEncounter[Index]):
            Spawner.SpawnedEvent.Subscribe(RandomEncounterMobSpawnDetected)
        for(Index:=0..AllNPC_GroundMobSpawner.Length-1, Spawner := AllNPC_GroundMobSpawner[Index]):
            Spawner.SpawnedEvent.Subscribe(handler_weaponsandmobs{MaybeGeneralMobsManager:=option{Self},SpawnerIndex:=Index,MobType:=GROUND_MOB}.HandlerDungeonMobSpawn)
            set AllNPC_GroundMobSpawner_IsBeingUsed = AllNPC_GroundMobSpawner_IsBeingUsed + array{-1}
            #Spawner.SpawnedEvent.Subscribe(MobSpawnDetected)
        for(Index:=0..AllNPC_DruidMobSpawner.Length-1, Spawner := AllNPC_DruidMobSpawner[Index]):
            set AllNPC_DruidMobSpawner_IsBeingUsed = AllNPC_DruidMobSpawner_IsBeingUsed + array{-1}
            Spawner.SpawnedEvent.Subscribe(handler_weaponsandmobs{MaybeGeneralMobsManager:=option{Self},SpawnerIndex:=Index,MobType:=DRUID_MOB}.HandlerDungeonMobSpawn)
            #Spawner.SpawnedEvent.Subscribe(MobSpawnDetected)
        for(Index:=0..AllNPC_PlantMobSpawner.Length-1, Spawner := AllNPC_PlantMobSpawner[Index]):
            set AllNPC_PlantMobSpawner_IsBeingUsed = AllNPC_PlantMobSpawner_IsBeingUsed + array{-1}
            Spawner.SpawnedEvent.Subscribe(handler_weaponsandmobs{MaybeGeneralMobsManager:=option{Self},SpawnerIndex:=Index,MobType:=PLANT_MOB}.HandlerDungeonMobSpawn)
            #Spawner.SpawnedEvent.Subscribe(MobSpawnDetected)
        for(Index:=0..AllNPC_MiniBossMobSpawner.Length-1, Spawner := AllNPC_MiniBossMobSpawner[Index]):
            set AllNPC_MiniBossMobSpawner_IsBeingUsed = AllNPC_MiniBossMobSpawner_IsBeingUsed + array{-1}
            Spawner.SpawnedEvent.Subscribe(handler_weaponsandmobs{MaybeGeneralMobsManager:=option{Self},SpawnerIndex:=Index,MobType:=MINIBOSS_MOB}.HandlerDungeonMobSpawn)



    CleanNPCMemory(Assignation : int) : void =
        if(MobsManager := AllMobsManager[Assignation]):
            Log("Cleaned assignation {Assignation}",LOG_LEVEL_NPCDEBUG)
            MobsManager.SetUse(false)

<#
using { /Fortnite.com/Devices }
using { /Verse.org/Simulation }
using { /UnrealEngine.com/Temporary/Diagnostics }
using { /Fortnite.com/Characters }
using { /UnrealEngine.com/Temporary/SpatialMath }

# See https://dev.epicgames.com/documentation/en-us/uefn/create-your-own-device-in-verse for how to create a verse device.

# A Verse-authored creative device that can be placed in a level
general_mobs_manager := class(creative_device):

    var TotalGameTime : int = 0

    var MaybeTestsManager : ?tests_manager = false
    DungeonDataBase : dungeon_database = dungeon_database{}

    @editable
    DirtSpawnEffect : creative_prop_asset = DefaultCreativePropAsset

    @editable
    HealthBarIsOrganic : logic = true

    @editable
    AllPropManipulator : []prop_manipulator_device = array{}

    @editable
    WeaponsManager : weapons_manager = weapons_manager{}

    var AllMobsManager : []mob = array{}
    var AllMobInfoToLoad : []tuple(transform,logic,int,int) = array{}

    #Ground spawners
    var AllNPC_GroundMobSpawner_IsBeingUsed : []int = array{}
    @editable
    AllNPC_GroundMobSpawner : []npc_spawner_device = array{}
    var AllNPC_DruidMobSpawner_IsBeingUsed : []int = array{}
    @editable
    AllNPC_DruidMobSpawner : []npc_spawner_device = array{}
    #@editable
    #AllNPC_DruidMobSpawner_Root : []creative_prop = array{}
    #Miniboss spawners
    var AllNPC_MiniBossMobSpawner_IsBeingUsed : []int = array{}
    @editable
    AllNPC_MiniBossMobSpawner : []npc_spawner_device = array{}
    #@editable
    #AllNPC_MiniBossMobSpawner_Root : []creative_prop = array{}
    
    #AllDungeons
    var Dungeon_TresureRegistration : [agent][]int = map{}
    var AllDungeons_IsPublic : []logic = array{}
    var AllDungeons_MobsSpawned : [][]int = array{}
    var AllDungeons_MobsDistributionToBeSpawn : [][]int = array{}
    var AllDungeons_NPCSpawnersInUse : [][]int = array{}

    var AllDungeons_MinibossIsDefeated : []logic = array{}
    var AllDungeons_MaybeStartingAgent : []?agent = array{}
    var AllDungeons_HasStarted : []logic = array{}
    var AllDungeons_AgentsWaitingToJoin : [][]agent = array{}
    var AllDungeons_IsReadyToPlay : []logic = array{}
    var AllDungeons_CountToStartTime : []int = array{}
    var AllDungeons_TotalTime : []int = array{}
    var AllDungeons_Buttons : [][]button_device = array{}
    var AllDungeons_Positions : [][]creative_prop = array{} 
    var AllDungeons_AgentsPlaying : [][]agent = array{}

    #Positions! 
    #0 Position: Entry transform
    #1 Position: Exit transform
    #2 Position: Boss spawner transform
    #3..X: Posible Mob spawner transform

    #Buttons!
    #0 Position: Entry button
    #1 Position: Exit button
    #2 Position: Main treasure button
    #3..X: Position: Other treasures button

    #Dungeon 0
    var Dungeon0_TotalTime : int = 0
    @editable
    Dungeon0_Name : string = "City sewer"
    @editable
    AllDungeon0_Positions : []creative_prop = array{}
    @editable
    AllDungeon0_Buttons : []button_device = array{}
    var AllDungeon0_NPCSpawnerUsed : []int = array{-1,-1,-1}

    var IsDungeon0BeingPlayed : logic = true
    var Dungeon0_MobsDistribution : []int = array{}
    var Dungeon0_MobsSpawned : []int = array{}
    var Dungeon0_AgentsPlaying : []agent = array{}
    var Dungeon0_AgentsWaiting : []agent = array{}

    #Dungeon 1
    var Dungeon1_TotalTime : int = 0
    @editable
    Dungeon1_Name : string = "Desert cave"
    @editable
    AllDungeon1_Positions : []creative_prop = array{}
    @editable
    AllDungeon1_Buttons : []button_device = array{}
    var AllDungeon1_NPCSpawnerUsed : []int = array{-1,-1,-1}

    var IsDungeon1BeingPlayed : logic = true
    var Dungeon1_MobsDistribution : []int = array{}
    var Dungeon1_MobsSpawned : []int = array{}
    var Dungeon1_AgentsPlaying : []agent = array{}
    var Dungeon1_AgentsWaiting : []agent = array{}

    #Dungeon 2
    var Dungeon2_TotalTime : int = 0
    @editable
    Dungeon2_Name : string = "Secret laboratory"
    @editable
    AllDungeon2_Positions : []creative_prop = array{}
    @editable
    AllDungeon2_Buttons : []button_device = array{}
    var AllDungeon2_NPCSpawnerUsed : []int = array{-1,-1,-1}

    var IsDungeon2Available : logic = true
    var Dungeon2_MobsDistribution : []int = array{}
    var Dungeon2_MobsSpawned : []int = array{}
    var Dungeon2_AgentsPlaying : []agent = array{}
    var Dungeon2_AgentsWaiting : []agent = array{}

    #Dungeon 3
    var Dungeon3_TotalTime : int = 0
    @editable
    Dungeon3_Name : string = "Jungle secrets"
    @editable
    AllDungeon3_Positions : []creative_prop = array{}
    @editable
    AllDungeon3_Buttons : []button_device = array{}
    var AllDungeon3_NPCSpawnerUsed : []int = array{-1,-1,-1}

    var IsDungeon3BeingPlayed : logic = true
    var Dungeon3_MobsDistribution : []int = array{}
    var Dungeon3_MobsSpawned : []int = array{}
    var Dungeon3_AgentsPlaying : []agent = array{}
    var Dungeon3_AgentsWaiting : []agent = array{}

    #Dungeon 4
    var Dungeon4_TotalTime : int = 0
    @editable
    Dungeon4_Name : string = "Amadeus I"
    @editable
    AllDungeon4_Positions : []creative_prop = array{}
    @editable
    AllDungeon4_Buttons : []button_device = array{}
    var AllDungeon4_NPCSpawnerUsed : []int = array{-1,-1,-1}

    var IsDungeon4BeingPlayed : logic = true
    var Dungeon4_MobsDistribution : []int = array{}
    var Dungeon4_MobsSpawned : []int = array{}
    var Dungeon4_AgentsPlaying : []agent = array{}
    var Dungeon4_AgentsWaiting : []agent = array{}

    #Dungeon 5
    var Dungeon5_TotalTime : int = 0
    @editable
    Dungeon5_Name : string = "Amadeus II"
    @editable
    AllDungeon5_Positions : []creative_prop = array{}
    @editable
    AllDungeon5_Buttons : []button_device = array{}
    var AllDungeon5_NPCSpawnerUsed : []int = array{-1,-1,-1}

    var IsDungeon5BeingPlayed : logic = true
    var Dungeon5_MobsDistribution : []int = array{}
    var Dungeon5_MobsSpawned : []int = array{}
    var Dungeon5_AgentsPlaying : []agent = array{}
    var Dungeon5_AgentsWaiting : []agent = array{}

    #Dungeon 6
    var Dungeon6_TotalTime : int = 0
    @editable
    Dungeon6_Name : string = "Amadeus III"
    @editable
    AllDungeon6_Positions : []creative_prop = array{}
    @editable
    AllDungeon6_Buttons : []button_device = array{}
    var AllDungeon6_NPCSpawnerUsed : []int = array{-1,-1,-1}

    var IsDungeon6BeingPlayed : logic = true
    var Dungeon6_MobsDistribution : []int = array{}
    var Dungeon6_MobsSpawned : []int = array{}
    var Dungeon6_AgentsPlaying : []agent = array{}
    var Dungeon6_AgentsWaiting : []agent = array{}


    #Dungeon 7
    var Dungeon7_TotalTime : int = 0
    @editable
    Dungeon7_Name : string = "Amadeus IV"
    @editable
    AllDungeon7_Positions : []creative_prop = array{}
    @editable
    AllDungeon7_Buttons : []button_device = array{}
    var AllDungeon7_NPCSpawnerUsed : []int = array{-1,-1,-1}

    var IsDungeon7BeingPlayed : logic = true
    var Dungeon7_MobsDistribution : []int = array{}
    var Dungeon7_MobsSpawned : []int = array{}
    var Dungeon7_AgentsPlaying : []agent = array{}
    var Dungeon7_AgentsWaiting : []agent = array{}

    #WEAPONS

    # Runs when the device is started in a running game
    OnBegin<override>()<suspends>:void=
        for(Player : GetPlayspace().GetPlayers(), Agent := agent[Player]):
            AddAgentToChestRegistration(Agent)

        InitOrganizationArrays()
        GetPlayspace().PlayerRemovedEvent().Subscribe(OnPlayerRemoved)
        GetPlayspace().PlayerAddedEvent().Subscribe(OnPlayerAdded)
        PrintWarning("TODO OnPlayerEliminated on general mobs manager for dungeons")
        
        CreateMobsManagers()
        InitNPCSpawners()
        InitSubscribers()
        MoveVisualEmpty()

        spawn:
            LoopGameTime()

    MoveVisualEmpty(): void =
        for(AllDungeonPositions : AllDungeons_Positions):
            for(Position : AllDungeonPositions):
                NewTranslationX:=Position.GetTransform().Translation.X
                NewTranslationY:=Position.GetTransform().Translation.Y
                NewTranslationZ:=Position.GetTransform().Translation.Z-LowVisualEmpty
                NewTranslation:=vector3{X:=NewTranslationX,Y:=NewTranslationY,Z:=NewTranslationZ}
                if(Position.TeleportTo[transform{Translation:=NewTranslation,Rotation:=Position.GetTransform().Rotation,Scale:=Position.GetTransform().Scale}]):

    LoopGameTime()<suspends> : void =
        PrintWarning("TODO: Loop game time")
        loop:
            Sleep(0.4)
            set TotalGameTime = TotalGameTime + 1
            CheckNPCSpawners()

    OnPlayerEliminated(Player : player) : void =
        if(Agent := agent[Player]):
            RemoveAgentFromAllDungeons(Agent)

    InitOrganizationArrays() : void =      

        for(Index:=0..AllPropManipulator.Length-1,Manipulator := AllPropManipulator[Index]):
            Manipulator.DamagedEvent.Subscribe(handler_weaponsandmobs{MaybeGeneralMobsManager:=option{Self}, Position:=Manipulator.GetTransform(),MobLevel := Index}.HandlerTestMobSpawn)

        #NPCSpawners
        set AllDungeons_NPCSpawnersInUse = array{AllDungeon0_NPCSpawnerUsed,AllDungeon1_NPCSpawnerUsed,AllDungeon2_NPCSpawnerUsed,
            AllDungeon3_NPCSpawnerUsed,AllDungeon4_NPCSpawnerUsed,AllDungeon5_NPCSpawnerUsed,AllDungeon6_NPCSpawnerUsed,
            AllDungeon7_NPCSpawnerUsed}

        #Positions
        set AllDungeons_Positions = array{AllDungeon0_Positions,AllDungeon1_Positions,AllDungeon2_Positions,AllDungeon3_Positions,
         AllDungeon4_Positions,AllDungeon5_Positions,AllDungeon6_Positions,AllDungeon7_Positions}

        #All agents playing
        set AllDungeons_AgentsPlaying = array{Dungeon0_AgentsPlaying,Dungeon1_AgentsPlaying,Dungeon2_AgentsPlaying,Dungeon3_AgentsPlaying,
            Dungeon4_AgentsPlaying,Dungeon5_AgentsPlaying,Dungeon6_AgentsPlaying,Dungeon7_AgentsPlaying}

        #All Dungeon MobsSpawned
        set AllDungeons_MobsSpawned = array{Dungeon0_MobsSpawned,Dungeon1_MobsSpawned,Dungeon3_MobsSpawned,Dungeon4_MobsSpawned,
            Dungeon5_MobsSpawned,Dungeon6_MobsSpawned,Dungeon7_MobsSpawned}

        #All Dungeon Mob Distribution
        set AllDungeons_MobsDistributionToBeSpawn = array{Dungeon0_MobsDistribution,Dungeon1_MobsDistribution,Dungeon2_MobsDistribution,
            Dungeon3_MobsDistribution,Dungeon4_MobsDistribution,Dungeon5_MobsDistribution,Dungeon6_MobsDistribution,
            Dungeon7_MobsDistribution}

        #All Dungeon Buttons
        set AllDungeons_Buttons = array{AllDungeon0_Buttons,AllDungeon1_Buttons,AllDungeon2_Buttons,AllDungeon3_Buttons,AllDungeon4_Buttons,
            AllDungeon5_Buttons,AllDungeon6_Buttons,AllDungeon7_Buttons}

        #All agents waiting
        set AllDungeons_AgentsWaitingToJoin = array{Dungeon0_AgentsWaiting,Dungeon1_AgentsWaiting,Dungeon2_AgentsWaiting,Dungeon3_AgentsWaiting,
            Dungeon4_AgentsWaiting,Dungeon5_AgentsWaiting,Dungeon6_AgentsWaiting,Dungeon7_AgentsWaiting}

        #Is ready to play
        for(Dungeon : AllDungeons_Positions, DummyAgent := agent[GetPlayspace().GetPlayers()[0]]):
            set AllDungeons_IsReadyToPlay = AllDungeons_IsReadyToPlay + array{true} #Is ready to play
            set AllDungeons_TotalTime = AllDungeons_TotalTime + array{99} #Time
            set AllDungeons_HasStarted = AllDungeons_HasStarted + array{false}
            set AllDungeons_IsPublic = AllDungeons_IsPublic + array{false}
            set AllDungeons_MaybeStartingAgent = AllDungeons_MaybeStartingAgent + array{option{DummyAgent}}
            set AllDungeons_CountToStartTime = AllDungeons_CountToStartTime + array{10}
            set AllDungeons_MinibossIsDefeated = AllDungeons_MinibossIsDefeated + array{false}
        for(Index:=0..AllDungeons_MaybeStartingAgent.Length-1, set AllDungeons_MaybeStartingAgent[Index]=false):
        
    AddAgentToChestRegistration(Agent : agent) : void =
        if(set Dungeon_TresureRegistration[Agent] = array{}):
        
    OnPlayerAdded(Player : player) : void =
        if(Agent := agent[Player]):
            AddAgentToChestRegistration(Agent)

    OnPlayerRemoved(Player : player) : void =
        if(Agent := player[Player]):
            set Dungeon_TresureRegistration=RemoveAgentFromTreasureRegistrationMap(Dungeon_TresureRegistration,Agent)
            RemoveAgentFromAllDungeons(Agent)

    RemoveAgentFromAllDungeons(Agent : agent) : void =
        #Removing agent from starting agent dungeon
        var FromDungeonRemoved : int = -1
        if(TestsManager := MaybeTestsManager?, DungeonUI := TestsManager.AllActionInfoUI[Agent]):
            DungeonUI.RemoveDungeonInfo()
        for(Index:=0..AllDungeons_MaybeStartingAgent.Length-1):
            if:
                MaybeStartingAgent := AllDungeons_MaybeStartingAgent[Index]
                ValidStartingAgent := MaybeStartingAgent?
                ValidStartingAgent = Agent
                set AllDungeons_MaybeStartingAgent[Index] = false
        for(IndexDungeon:=0..AllDungeons_AgentsPlaying.Length-1, DungeonAgents :=  AllDungeons_AgentsPlaying[IndexDungeon]):
            for(IndexAgent:=0..DungeonAgents.Length-1, AgentPlaying := DungeonAgents[IndexAgent], AgentPlaying = Agent):
                if:
                    Aux_DungeonAgents := DungeonAgents.RemoveElement[IndexAgent]
                    set AllDungeons_AgentsPlaying[IndexDungeon] = Aux_DungeonAgents
                then:
                    set FromDungeonRemoved = IndexDungeon
        for(IndexDungeon:=0..AllDungeons_AgentsWaitingToJoin.Length-1, DungeonAgents :=  AllDungeons_AgentsWaitingToJoin[IndexDungeon]):
            for(IndexAgent:=0..DungeonAgents.Length-1, AgentPlaying := DungeonAgents[IndexAgent], AgentPlaying = Agent):
                if:
                    Aux_DungeonAgents := DungeonAgents.RemoveElement[IndexAgent]
                    set AllDungeons_AgentsWaitingToJoin[IndexDungeon] = Aux_DungeonAgents
                then:
                    set FromDungeonRemoved = IndexDungeon
        CheckDungeonEnd(FromDungeonRemoved)

    RemoveAgentSpecificFromDungeon(Agent : agent, NumDungeon : int) : void =
        if(DungeonAgents :=  AllDungeons_AgentsPlaying[NumDungeon]):
            for(IndexAgent:=0..DungeonAgents.Length-1, AgentPlaying := DungeonAgents[IndexAgent], AgentPlaying = Agent):
                if:
                    Aux_DungeonAgents := DungeonAgents.RemoveElement[IndexAgent]
                    set AllDungeons_AgentsPlaying[NumDungeon] = Aux_DungeonAgents
                CheckDungeonEnd(NumDungeon)

    InitSubscribers() : void =
        for(IndexDungeon:=0..AllDungeons_Buttons.Length-1, DungeonButtons :=  AllDungeons_Buttons[IndexDungeon]):
            for(IndexButton:=0..DungeonButtons.Length-1, DungeonButton := DungeonButtons[IndexButton]):
                DungeonButton.InteractedWithEvent.Subscribe(handler_weaponsandmobs{MaybeGeneralMobsManager:=option{Self}, DungeonNumber:=IndexDungeon, DungeonButtonType:=IndexButton}.HandlerDungeonButton)
        
    TestHandlerActionUI(Agent : agent) : void =
        spawn:
            ActionUI(Agent)

    ActionUI(Agent : agent)<suspends> : void =
        #Print("Action detected")

    #Executed on SpawnedEvent
    MobSpawnDetected(MobAgent : agent) : void =
        var Index:int=0
        if:
            MobInfoToLoad:=AllMobInfoToLoad[0]
            MobInfoToLoad(1)<>-1
            MobCharacter := MobAgent.GetFortCharacter[] 
        then:
            loop:
                if:
                    MobManager := AllMobsManager[Index]
                    not MobManager.IsBeingUsed?
                then:
                    #Print("Found valid MobManager on {Index}")
                    MobManager.SetUse(true)
                    MobManager.SetStats(MobInfoToLoad(2),MobInfoToLoad(3))
                    MobManager.SetAgentAndCharacter(MobAgent, MobCharacter)
                    MobManager.Init()
                    if:
                        NewArray := AllMobInfoToLoad.RemoveElement[0]
                        set AllMobInfoToLoad = NewArray
                    break
                else:
                    set Index = Index + 1
                if(Index>AllMobsManager.Length-1):
                    PrintFatalError("FATAL ERROR, MOB MANAGER NOT FOUND")
                    break
        else:
            PrintFatalError("On MobSpawnDetected, MobInfoToLoad is invalid")

    TeleportPlayerToDungeon(Agent : agent, NumDungeon : int) : void =
        #Teleport all agents
        if:
            TestsManager:=MaybeTestsManager?
            AllDungeonPosition:=AllDungeons_Positions[NumDungeon]
            StartPositionProp:=AllDungeonPosition[1]
            TranslationProp:=StartPositionProp.GetTransform().Translation
            RotationProp:=StartPositionProp.GetTransform().Rotation
        then:
            NewTranslation:vector3=vector3{X:=TranslationProp.X,Y:=TranslationProp.Y,Z:=TranslationProp.Z+UpVisualEmpty}
            NewSpawnPosition:=transform{Translation:=NewTranslation,Rotation:=RotationProp,Scale:=StartPositionProp.GetTransform().Scale}
            if(Character := Agent.GetFortCharacter[]):                 
                if:
                    Character.TeleportTo[NewTranslation,RotationProp]
                else:
                    PrintWarning("TODO, safe teleport to other place if it fails")
                if(DungeonUI := TestsManager.AllDungeonUI[Agent]):
                    DungeonUI.Remove_CanvasWithoutMessage()
                if(ActionInfoUI := TestsManager.AllActionInfoUI[Agent]):
                    ActionInfoUI.SetDungeon_TimeInfo("Starting")
                    ActionInfoUI.SetDungeon_ChestInfo("Prepare for battle")
                #Print("Player teleported")


    ExecuteDungeon_MobSpawn(NumDungeon :int, SpawnPosition : transform, IsBoss : logic, MobType : int, Level : int)<suspends> : void =
        MobInfo := (SpawnPosition,IsBoss,MobType,Level)
        set AllMobInfoToLoad = AllMobInfoToLoad + array{MobInfo} #Save the new info to be loaded
        ##Print("Preparing dungeon spawn")
        ##Print("-----------------------")
        if:
            AllNPCSpawnersInUse:=AllDungeons_NPCSpawnersInUse[NumDungeon]
            IndexSpawnerToUse:=AllNPCSpawnersInUse[MobType]
        then:
            var NPCSpawner_ToUse : npc_spawner_device = npc_spawner_device{}
            #var NPCSpawnerRoot_ToUse : creative_prop = creative_prop{}
            if(MobType = GROUND_MOB, Spawner := AllNPC_GroundMobSpawner[IndexSpawnerToUse], <#Root := AllNPC_GroundMobSpawner_Root[IndexSpawnerToUse],#> set NPCSpawner_ToUse = Spawner <#set NPCSpawnerRoot_ToUse = Root#>):
                #Print("GROUND_MOB Spawner used: {IndexSpawnerToUse}")
            if(MobType = DRUID_MOB, Spawner := AllNPC_DruidMobSpawner[IndexSpawnerToUse], <#Root := AllNPC_DruidMobSpawner_Root[IndexSpawnerToUse],#> set NPCSpawner_ToUse = Spawner<# set NPCSpawnerRoot_ToUse = Root#>):
                #Print("DRUID_MOB Spawner used: {IndexSpawnerToUse}")
            NewTranslation:vector3=vector3{X:=SpawnPosition.Translation.X,Y:=SpawnPosition.Translation.Y,Z:=SpawnPosition.Translation.Z+UpVisualEmpty}
            NewSpawnPosition:=transform{Translation:=NewTranslation,Rotation:=SpawnPosition.Rotation,Scale:=SpawnPosition.Scale}
            if(NPCSpawner_ToUse.TeleportTo[NewSpawnPosition]):
                Sleep(0.5)
                SpawnTuple:=SpawnProp(DirtSpawnEffect,NewSpawnPosition)
                MaybeProp:=SpawnTuple(0)
                NPCSpawner_ToUse.Spawn()
                if(ValidProp:=MaybeProp?):
                    ##Print("Spawned prop")
                    spawn:
                        DestroyEffect(ValidProp)
                ##Print("Spawned mob")
            else:
                PrintFatalError("On ExecuteSpawn trying to teleport NPC Spawner to {NewTranslation.X},{NewTranslation.Y},{NewTranslation.Z}")

    #We calcule the new mob stats here and prepare the prespawn so the mob can load the stats right after the spawn
    ExecuteRandomEncounter_MobSpawn(SpawnPosition : transform, IsBoss : logic, MobType : int, Level : int)<suspends> : void =
        MobInfo := (SpawnPosition,IsBoss,MobType,Level)
        set AllMobInfoToLoad = AllMobInfoToLoad + array{MobInfo} #Save the new info to be loaded
        #Print("Preparing spawn")

        if(MobInfo(2)=GROUND_MOB):
            TupleResult:=FindSpawnerToBeUse(AllNPC_GroundMobSpawner_IsBeingUsed,AllNPC_GroundMobSpawner<#,AllNPC_GroundMobSpawner_Root#>)
            if:
                TupleResult(0)<>-1
                set AllNPC_GroundMobSpawner_IsBeingUsed[TupleResult(0)] = -1
                NPCSpawner:=TupleResult(1)?
                #NPCSpawner_Root:=TupleResult(2)?
            then:
                NewTranslation:vector3=vector3{X:=SpawnPosition.Translation.X,Y:=SpawnPosition.Translation.Y,Z:=SpawnPosition.Translation.Z+UpVisualEmpty}
                NewSpawnPosition:=transform{Translation:=NewTranslation,Rotation:=SpawnPosition.Rotation,Scale:=SpawnPosition.Scale}

                if(NPCSpawner.TeleportTo[NewSpawnPosition]):
                    Sleep(0.5)
                    ##Print("Spawning mob on position {NewTranslation.X},{NewTranslation.Y},{NewTranslation.Z}")
                    NPCSpawner.Spawn()
                else:
                    PrintFatalError("On ExecuteSpawn trying to teleport NPC Spawner to {NewTranslation.X},{NewTranslation.Y},{NewTranslation.Z}")

    DeployDungeonMobMultiSpawn(NumDungeon : int, PositionProp : creative_prop, MobsToSpawn : int)<suspends> : void =
        AllEnemiesToSpawn:=DungeonDataBase.GetEnemyWave(MobsToSpawn)
        Position:=PositionProp.GetTransform()

        MobLevel := DungeonDataBase.GetDungeonMobLevel(NumDungeon)

        for(Index:=0..AllEnemiesToSpawn.Length-1, EnemyType := AllEnemiesToSpawn[Index]):
            ExecuteDungeon_MobSpawn(NumDungeon,Position,false,EnemyType,NumDungeon)
            Sleep(0.25)

    DestroyEffect(CreativeProp : creative_prop)<suspends> : void =
        #Print("Destroying effect spawned")
        Sleep(5.0)
        CreativeProp.Dispose()

    SpawnMiniBoss(NumDungeon : int, PositionProp : creative_prop) : void =
        #Print("TODO")


    FindSpawnerToBeUse(AllUseController : []int, AllNPCSpawners : []npc_spawner_device<#, AllNPC_Roots : []creative_prop#>) : tuple(int,?npc_spawner_device<#,?creative_prop#>) =
        var Index : int = 0     
        loop:
            if:
                TimeWasInUse := AllUseController[Index]
                NPCSpawner := AllNPCSpawners[Index]
                #NPCSpawner_Root := AllNPC_Roots[Index] 
            then:
                if(TimeWasInUse<>-1):
                    #Print("Found valid MobManager on {Index}")
                    return (Index,option{NPCSpawner}<#,option{NPCSpawner_Root}#>)
                else:
                    set Index = Index + 1
            else:
                PrintFatalError("On FindSpawnerToBeUse. Index {Index} Invalid!")
            if(Index>AllUseController.Length-1):
                PrintFatalError("FATAL ERROR, MOB MANAGER NOT FOUND")
                break
        return(-1,false<#,false#>)

    ResetChestRegistration(Agent : agent) : void =
        if:
            set Dungeon_TresureRegistration[Agent] = array{}
        then:
            #Print("Treasure Registration reseted")

    SpawnDungeonTime(NumDungeon : int)<suspends> : void =
        MaxTimeToPlay:=DungeonDataBase.GetDungeonMaxTime(NumDungeon)
        loop:
            if:
                TotalTime := AllDungeons_TotalTime[NumDungeon] - 1
                NumAgentsPlaying := AllDungeons_AgentsPlaying[NumDungeon].Length
                set AllDungeons_TotalTime[NumDungeon] = TotalTime
                TestsManager := MaybeTestsManager?
            then:
                #Action UI
                if(AllAgents := AllDungeons_AgentsPlaying[NumDungeon]):
                    for(Agent : AllAgents):
                        if(ActionInfoUI := TestsManager.AllActionInfoUI[Agent]):
                            if(AllTreasuresOpened:=Dungeon_TresureRegistration[Agent].Length,AllTreasures:=AllDungeons_Buttons[NumDungeon].Length-2):
                                ActionInfoUI.SetDungeon_TimeInfo("Limit time: {TotalTime}")
                                ActionInfoUI.SetDungeon_ChestInfo("Chests opened: {AllTreasuresOpened}/{AllTreasures}")
                #Checking for Mob Spawn
                CheckDungeonMobSpawn(NumDungeon)
                #Dungeon UI
                if:
                    AllButtons := AllDungeons_Buttons[NumDungeon]
                    EnterButton := AllButtons[0]
                then:
                    #Dungeon UI
                    for(Player : GetPlayspace().GetPlayers(), Agent := player[Player], DungeonUI := TestsManager.AllDungeonUI[Agent], DungeonUI.Opened_NumDungeon=NumDungeon):
                        DungeonUI.UpdateUI()
                    #BUTTON JOIN
                    if(IsPublic := AllDungeons_IsPublic[NumDungeon]):
                        RemainingTime:=DungeonDataBase.CalculeDungeon_RemainingTimeToJoin(NumDungeon,TotalTime)
                        if:
                            RemainingTime<=0
                            set AllDungeons_IsReadyToPlay[NumDungeon] = false    
                        then:
                            EnterButton.SetInteractionText(BasicMessage("There is somebody playing this dungeon!"))
                        else:
                            if(IsPublic?):
                                EnterButton.SetInteractionText(BasicMessage("Join and help! {RemainingTime}"))
                            else:
                                EnterButton.SetInteractionText(BasicMessage("Private game. Wait until the current player finish"))

                if(NumAgentsPlaying<=0, set AllDungeons_IsReadyToPlay[NumDungeon] = true):
                    if(AllButtons := AllDungeons_Buttons[NumDungeon], EnterButton := AllButtons[0]):
                        EnterButton.SetInteractionText(BasicMessage("Clean dungeon"))
                    break
                if(TotalTime<=0):
                    TimeEnded(NumDungeon)
            else:
                PrintFatalError("On SpawnDungeonTime: if failed")
                break
            Sleep(1.0)

    CheckDungeonMobSpawn(NumDungeon : int) : void =
        ##Print("Checking to spawn mobs on {NumDungeon} A")
        if:
            AllPositions:=AllDungeons_Positions[NumDungeon]
            AllAgentsPlaying:=AllDungeons_AgentsPlaying[NumDungeon]
            var AllMobsSpawned : []int =AllDungeons_MobsSpawned[NumDungeon]
            AllMobsDistributionToBeSpawn:=AllDungeons_MobsDistributionToBeSpawn[NumDungeon]
        then:
            ##Print("Checking to spawn mobs on {NumDungeon} B")
            <#
            if(AllNPCsDungeon:=AllDungeons_MobsDistributionToBeSpawn[NumDungeon], AllNPCsDungeonSpawned:=AllDungeons_MobsSpawned[NumDungeon]):
                for(Index:=0..AllNPCsDungeon.Length-1, Distribution := AllNPCsDungeon[Index], Spawned := AllNPCsDungeonSpawned[Index]):
                    #Print("On position {Index} will be spawned {Distribution} MOBS. and there han been spawned {Spawned} MOBS")#>
            for(Index:=2..AllPositions.Length-1):
                if:
                    PositionProp := AllPositions[Index]
                    NumMobsToBeSpawned := AllMobsDistributionToBeSpawn[Index]
                    MobsSpawned := AllMobsSpawned[Index]
                    MobsSpawned < NumMobsToBeSpawned
                then:
                    var IsThereCharacter_Nearby : logic = false
                    PositionTranslation:=PositionProp.GetTransform().Translation
                    PositionTranslationToCalcule:=vector3{X:=PositionTranslation.X,Y:=PositionTranslation.Y,Z:=PositionTranslation.Z+UpVisualEmpty}
                    for(AgentPlaying : AllAgentsPlaying, CharacterPlaying := AgentPlaying.GetFortCharacter[]):
                        DistanceFromPoint := Distance(CharacterPlaying.GetTransform().Translation,PositionTranslationToCalcule)
                        if(DistanceFromPoint<DungeonDataBase.DungeonMobDistanceDetection):
                            set IsThereCharacter_Nearby = true
                            #Print("Player nearby at {DistanceFromPoint}, trying to spawn {NumMobsToBeSpawned} Mobs")
                    if:
                        IsThereCharacter_Nearby?
                        set AllMobsSpawned[Index]=NumMobsToBeSpawned
                        set AllDungeons_MobsSpawned[NumDungeon]=AllMobsSpawned
                    then:
                        spawn:
                            DeployDungeonMobMultiSpawn(NumDungeon,PositionProp,NumMobsToBeSpawned)

    CheckNPCSpawners() : void =
        #GROUND SPAWNER
        for(Index:=0..AllNPC_GroundMobSpawner_IsBeingUsed.Length-1):
            if:
                GroundMobSpawner_IsBeingUsed:=AllNPC_GroundMobSpawner_IsBeingUsed[Index]
                NPCMobSpawner:=AllNPC_GroundMobSpawner[Index]
                GroundMobSpawner_IsBeingUsed <> -1
                (GroundMobSpawner_IsBeingUsed + MAX_NPCMOBSPAWNER_TIMEUSE) < TotalGameTime
            then:
                PrintWarning("CheckNPCSpawners: MAX NPC TIME USE REACHED!")
                if:
                    set AllNPC_GroundMobSpawner_IsBeingUsed[Index] = -1
                then:
                    NPCMobSpawner.DespawnAll(false)
                    #NPCMobSpawner.Disable()
        #DRUID SPAWNER
        for(Index:=0..AllNPC_DruidMobSpawner_IsBeingUsed.Length-1):
            if:
                DruidMobSpawner_IsBeingUsed:=AllNPC_DruidMobSpawner_IsBeingUsed[Index]
                NPCMobSpawner:=AllNPC_DruidMobSpawner[Index]
                DruidMobSpawner_IsBeingUsed <> -1
                (DruidMobSpawner_IsBeingUsed + MAX_NPCMOBSPAWNER_TIMEUSE) < TotalGameTime
            then:
                PrintWarning("CheckNPCSpawners: MAX NPC TIME USE REACHED!")
                if:
                    set AllNPC_DruidMobSpawner_IsBeingUsed[Index] = -1
                then:
                    NPCMobSpawner.DespawnAll(false)
                    #NPCMobSpawner.Disable()
        #MINIBOSS SPAWNER
        for(Index:=0..AllNPC_MiniBossMobSpawner_IsBeingUsed.Length-1):
            if:
                MiniBossMobSpawner_IsBeingUsed:=AllNPC_MiniBossMobSpawner_IsBeingUsed[Index]
                NPCMobSpawner:=AllNPC_MiniBossMobSpawner[Index]
                MiniBossMobSpawner_IsBeingUsed <> -1
                (MiniBossMobSpawner_IsBeingUsed + MAX_NPCMOBSPAWNER_TIMEUSE) < TotalGameTime
            then:
                PrintWarning("CheckNPCSpawners: MAX NPC TIME USE REACHED!")
                if:
                    set AllNPC_MiniBossMobSpawner_IsBeingUsed[Index] = -1
                then:
                    NPCMobSpawner.DespawnAll(false)
                    #NPCMobSpawner.Disable()
                    
                


    FindDungeonMobSpawners(NumDungeon : int) : void =
        #Ground mob spawner
        var Index_GroundMobSpawner : int = 0
        loop:
            if(TimeWasBeingUse:=AllNPC_GroundMobSpawner_IsBeingUsed[Index_GroundMobSpawner], TimeWasBeingUse=-1):
                break
            else:
                set Index_GroundMobSpawner = Index_GroundMobSpawner + 1
            if(Index_GroundMobSpawner>AllNPC_GroundMobSpawner_IsBeingUsed.Length-1):
                set Index_GroundMobSpawner = -1
                break
        #Druid mob spawner
        var Index_DruidMobSpawner : int = 0
        loop:
            if(TimeWasBeingUse:=AllNPC_DruidMobSpawner_IsBeingUsed[Index_DruidMobSpawner], TimeWasBeingUse=-1):
                break
            else:
                set Index_DruidMobSpawner = Index_DruidMobSpawner + 1
            if(Index_DruidMobSpawner>AllNPC_DruidMobSpawner_IsBeingUsed.Length-1):
                set Index_DruidMobSpawner = -1
                break
        #MiniBoss mob spawner
        var Index_MiniBossMobSpawner : int = 0
        loop:
            if(TimeWasBeingUse:=AllNPC_MiniBossMobSpawner_IsBeingUsed[Index_MiniBossMobSpawner], TimeWasBeingUse=-1):
                break
            else:
                set Index_MiniBossMobSpawner = Index_MiniBossMobSpawner + 1
            if(Index_MiniBossMobSpawner>AllNPC_MiniBossMobSpawner_IsBeingUsed.Length-1):
                set Index_MiniBossMobSpawner = -1
                break

        if(Index_GroundMobSpawner=-1 or Index_DruidMobSpawner=-1 or Index_MiniBossMobSpawner=-1):
            PrintFatalError("On FindDungeonMobSpawners no found. Invalid. Index_GroundMobSpawner: {Index_GroundMobSpawner}. Index_DruidMobSpawner: {Index_DruidMobSpawner}. Index_MiniBossMobSpawner: {Index_MiniBossMobSpawner}")
        if:
            set AllDungeons_NPCSpawnersInUse[NumDungeon] = array{Index_GroundMobSpawner,Index_DruidMobSpawner,Index_MiniBossMobSpawner}
            set AllNPC_GroundMobSpawner_IsBeingUsed[Index_GroundMobSpawner] = TotalGameTime
            set AllNPC_DruidMobSpawner_IsBeingUsed[Index_DruidMobSpawner] = TotalGameTime
            set AllNPC_MiniBossMobSpawner_IsBeingUsed[Index_MiniBossMobSpawner] = TotalGameTime
            GroundMobSpawner:=AllNPC_GroundMobSpawner[Index_GroundMobSpawner]
            DruidMobSpawner:=AllNPC_DruidMobSpawner[Index_DruidMobSpawner]
            MiniBossMobSpawner:=AllNPC_MiniBossMobSpawner[Index_MiniBossMobSpawner]
        then:
            #GroundMobSpawner.Enable()
            #DruidMobSpawner.Enable()
            #MiniBossMobSpawner.Enable()
            GroundMobSpawner.DespawnAll(false)
            DruidMobSpawner.DespawnAll(false)
            MiniBossMobSpawner.DespawnAll(false)
        else:
            PrintFatalError("NPC Spawners not found!")
        



    TimeEnded(NumDungeon : int) : void =
        #Print("The time to play the dungeon has ended!")
        if(AllAgentsPlaying:=AllDungeons_AgentsPlaying[NumDungeon]):
            for(AgentPlaying : AllAgentsPlaying):
                QuickAgentFromDungeon(AgentPlaying,NumDungeon)

    QuickAgentFromDungeon(Agent : agent, NumDungeon : int) : void =
        #Print("Trying to abandon dungeon {NumDungeon}")
        ResetChestRegistration(Agent)
        if(TestsManager := MaybeTestsManager?):
            if(ActionInfoUI := TestsManager.AllActionInfoUI[Agent]):
                ActionInfoUI.RemoveDungeonInfo()
        RemoveAgentSpecificFromDungeon(Agent,NumDungeon)
        if(AllPositions:=AllDungeons_Positions[NumDungeon], ExitPosition:=AllPositions[0]):

            RotationExit:=ExitPosition.GetTransform().Rotation
            NewTranslation:vector3=vector3{X:=ExitPosition.GetTransform().Translation.X,Y:=ExitPosition.GetTransform().Translation.Y,Z:=ExitPosition.GetTransform().Translation.Z+UpVisualEmpty}
            NewExitPosition:=transform{Translation:=NewTranslation,Rotation:=RotationExit,Scale:=ExitPosition.GetTransform().Scale}

            if(Character:=Agent.GetFortCharacter[],Character.TeleportTo[NewTranslation,RotationExit]):
                PrintWarning("TODO, safe teleport to other place if it fails")
            else:
                PrintFatalError("Teleportation has failed")
        else:
            PrintFatalError("Position not found")

    CheckDungeonEnd(NumDungeon : int) : void =
        if(TestsManager := MaybeTestsManager?,AllPlayers:=AllDungeons_AgentsPlaying[NumDungeon]):
            if(AllPlayers.Length<=0):
                ResetDungeon(NumDungeon)
                if(AllButtons:=AllDungeons_Buttons[NumDungeon],EntryButton:=AllButtons[0]):
                    EntryButton.SetInteractionText(BasicMessage("Clean dungeon"))
                    TestsManager.UpdateAll_DungeonUI(NumDungeon)

    ResetDungeon(NumDungeon : int) : void =        
        #Delete all currect NPCS, despawn and disable
        if:
            NPCSpawnersInUse:=AllDungeons_NPCSpawnersInUse[NumDungeon]
            GroundSpawnerInUse := AllNPC_GroundMobSpawner[NPCSpawnersInUse[0]]
            DruidSpawnerInUse := AllNPC_DruidMobSpawner[NPCSpawnersInUse[1]]
            MiniBossSpawnerInUse := AllNPC_MiniBossMobSpawner[NPCSpawnersInUse[2]]
            set AllNPC_GroundMobSpawner_IsBeingUsed[NPCSpawnersInUse[0]] = -1
            set AllNPC_DruidMobSpawner_IsBeingUsed[NPCSpawnersInUse[1]] = -1
            set AllNPC_MiniBossMobSpawner_IsBeingUsed[NPCSpawnersInUse[1]] = -1
        then:
            GroundSpawnerInUse.DespawnAll(false)
            DruidSpawnerInUse.DespawnAll(false)
            MiniBossSpawnerInUse.DespawnAll(false)
            #GroundSpawnerInUse.Disable()
            #DruidSpawnerInUse.Disable()
            #MiniBossSpawnerInUse.Disable()
        if:
            TestsManager:=MaybeTestsManager?
            AllButtons := AllDungeons_Buttons[NumDungeon]
            EntryButton := AllButtons[0]
            set AllDungeons_NPCSpawnersInUse[NumDungeon] = array{-1,-1,-1}
            set AllDungeons_MobsSpawned[NumDungeon] = array{}
            set AllDungeons_MinibossIsDefeated[NumDungeon] = false
            set AllDungeons_CountToStartTime[NumDungeon] = 10
            set AllDungeons_IsPublic[NumDungeon] = false
            set AllDungeons_MaybeStartingAgent[NumDungeon] = false
            set AllDungeons_AgentsWaitingToJoin[NumDungeon] = array{}
            set AllDungeons_IsReadyToPlay[NumDungeon] = true
            set AllDungeons_TotalTime[NumDungeon] = 1
            set AllDungeons_AgentsPlaying[NumDungeon] = array{}
        then:
            if(AgentsWaiting:=AllDungeons_AgentsWaitingToJoin[NumDungeon]):
                #Print("Reset dungeon completed, agents waiting: {AgentsWaiting.Length}")
            TestsManager.UpdateAll_DungeonUI(NumDungeon)
            EntryButton.SetInteractionText(BasicMessage("Play dungeon"))


    #Start dungeon
    StartDungeon(Agent : agent, NumDungeon : int): void =
        #Print("Starting Dungeon: {NumDungeon}")
        #Other check, that the starting player is available
        if(TestsManager := MaybeTestsManager?, MaybeStartingAgent:=AllDungeons_MaybeStartingAgent[NumDungeon], ValidStartingAgent := MaybeStartingAgent?):
            MaxTimeToPlay:=DungeonDataBase.GetDungeonMaxTime(NumDungeon)
            if:
                AllAgentsWaitingToJoin := AllDungeons_AgentsWaitingToJoin[NumDungeon]
                set AllDungeons_AgentsPlaying[NumDungeon] = array{} #Clear memory
                set AllDungeons_AgentsPlaying[NumDungeon] = AllAgentsWaitingToJoin #Set the new agents
                set AllDungeons_TotalTime[NumDungeon] = MaxTimeToPlay #TotaTime
                NumPositions := AllDungeons_Positions[NumDungeon]
            then:
                var ArrayMobsSpawned : []int = array{}
                for(Index:=0..NumPositions.Length-1):
                    set ArrayMobsSpawned = ArrayMobsSpawned + array{0}
                MobsDistribution := DungeonDataBase.GetNumMobsToBeSpawn(NumDungeon,NumPositions.Length)
                #Consume one try to starting player and prepare mob spawns
                if:
                    Player:=player[Agent]
                    PlayerData:=TestsManager.PlayersDataManager.GetPlayerData[Player]
                    PlayerData.Dungeon_Accumulated>0
                    set AllDungeons_MobsDistributionToBeSpawn[NumDungeon] = MobsDistribution
                    set AllDungeons_MobsSpawned[NumDungeon] = ArrayMobsSpawned
                then:
                    #if(AllNPCsDungeon:=AllDungeons_MobsDistributionToBeSpawn[NumDungeon], AllNPCsDungeonSpawned:=AllDungeons_MobsSpawned[NumDungeon]):
                    #    for(Index:=0..AllNPCsDungeon.Length-1, Distribution := AllNPCsDungeon[Index], Spawned := AllNPCsDungeonSpawned[Index]):
                    #        #Print("On position {Index} will be spawned {Distribution} MOBS. and there han been spawned {Spawned} MOBS")
                    #if:
                        #AllNPCsDungeonSpawned:=AllDungeons_MobsSpawned[NumDungeon]
                    #then:
                        ##Print("AllNPCsDungeonSpawned.Length is {AllNPCsDungeonSpawned.Length}")
                        #for(Index:=0..MobsDistribution.Length-1, Distribution := MobsDistribution[Index], Spawned := AllNPCsDungeonSpawned[Index]):
                        #    #Print("On position {Index} will be spawned {Distribution} MOBS. and there han been spawned {Spawned} MOBS")
                    #else:
                    #    PrintFatalError("AllDungeons_MobsSpawned {NumDungeon} not found")

                    NewPlayerData:=TestsManager.PlayersDataManager.ConsumeDungeonAccumulated(PlayerData)
                    TestsManager.PlayersDataManager.SaveNewData(Player,PlayerData,NewPlayerData)
                    FindDungeonMobSpawners(NumDungeon)
                    #Start dungeon timer
                    spawn:
                        SpawnDungeonTime(NumDungeon)
                    #Teleport all agents
                    for(AgentToTeleport : AllAgentsWaitingToJoin):     
                        TeleportPlayerToDungeon(AgentToTeleport,NumDungeon)  
            else:
                PrintFatalError("StartDungeon has failed!")          

    TryJoinPlayerTo_InitiatedDungeon(Agent : agent, NumDungeon : int) : void =
        #Game has no started, join to the waiting room
        var IsNotAlreadyInt : logic = false
        if:
            var AllAgentsPlaying : []agent = AllDungeons_AgentsPlaying[NumDungeon]
            var AllAgentsWaitingToJoin : []agent = AllDungeons_AgentsWaitingToJoin[NumDungeon]
            HasStarted := AllDungeons_HasStarted[NumDungeon]
            IsPublic := AllDungeons_IsPublic[NumDungeon]
            TotalTime := AllDungeons_TotalTime[NumDungeon]
        then:
            for(JoinedAgent : AllAgentsWaitingToJoin, JoinedAgent = Agent):
                set IsNotAlreadyInt = true
            for(JoinedAgent : AllAgentsWaitingToJoin, JoinedAgent = Agent):
                set IsNotAlreadyInt = true
            if(not IsNotAlreadyInt?):
                if(IsPublic?):
                    if(HasStarted?):
                        #Has started, can join?
                        RemainingTime:=DungeonDataBase.CalculeDungeon_RemainingTimeToJoin(NumDungeon,TotalTime)
                        if:
                            RemainingTime>0
                        then:
                            set AllAgentsPlaying = AllAgentsPlaying + array{Agent}
                            set AllAgentsWaitingToJoin = AllAgentsWaitingToJoin + array{Agent}
                            if(set AllDungeons_AgentsPlaying[NumDungeon] = AllAgentsPlaying, set AllDungeons_AgentsWaitingToJoin[NumDungeon] = AllAgentsWaitingToJoin):
                                TeleportPlayerToDungeon(Agent,NumDungeon)
                        else:
                            #Print("Is too late to join!")
                    else:
                        #Has no started, so join him to the waiting room
                        set AllAgentsPlaying = AllAgentsPlaying + array{Agent}
                        set AllAgentsWaitingToJoin = AllAgentsWaitingToJoin + array{Agent}
                        if(set AllDungeons_AgentsPlaying[NumDungeon] = AllAgentsPlaying, set AllDungeons_AgentsWaitingToJoin[NumDungeon] = AllAgentsWaitingToJoin):

    DungeonUI_PlayButtonPressed(Agent : agent, NumDungeon : int) : void =
        if(TestsManager := MaybeTestsManager?, DungeonUI:=TestsManager.AllDungeonUI[Agent]):
            if(IsReady := AllDungeons_IsReadyToPlay[NumDungeon], IsReady?):
                if:
                    var AllAgentsPlaying : []agent = AllDungeons_AgentsPlaying[NumDungeon]
                    var AllAgentsWaitingToPlay : []agent = AllDungeons_AgentsWaitingToJoin[NumDungeon]
                then:
                    if(AllAgentsPlaying.Length<=0, AllAgentsWaitingToPlay.Length<=0): 
                        #There is no playes playing or waiting, trying to start a game
                        #Start if only has permisions acumulated
                        if:
                            Player:=player[Agent]
                            PlayerData:=TestsManager.PlayersDataManager.GetPlayerData[Player]
                            PlayerData.Dungeon_Accumulated>0
                        then:
                            #Print("Runing countdown")
                            spawn:
                                AddToDungeonAndStartCountDown(Agent,NumDungeon)
                            DungeonUI.UpdateUI()
                        else:
                            spawn:
                                DungeonUI.RunNoCleaningPermit()
                    else if(MaybeAgent:=AllDungeons_MaybeStartingAgent[NumDungeon],ValidStartingAgent:=MaybeAgent?,ValidStartingAgent=Agent):
                        PrintWarning("TODO Ask for help on DungeonUI_PlayButtonPressed")
                        if(set AllDungeons_IsPublic[NumDungeon]=true):
                            TestsManager.UpdateAll_DungeonUI(NumDungeon)
                            #Print("Now is public")
                        DungeonUI.UpdateUI()
                    else if(HasStarted:=AllDungeons_HasStarted[NumDungeon]):
                        if(not HasStarted?): 
                            TryJoinPlayerTo_InitiatedDungeon(Agent,NumDungeon)
                        else:
                        #Check if the player can join
                        MaxTimeToPlay:=TestsManager.GeneralMobsManager.DungeonDataBase.GetDungeonMaxTime(NumDungeon)
                        if:
                            TotalTime:=AllDungeons_TotalTime[NumDungeon]
                            TotalTime<MaxTimeToPlay-(DungeonDataBase.MaxTimeToEnter)
                        then:
                            RemaingTimeToJoin:=TotalTime-(MaxTimeToPlay-(DungeonDataBase.MaxTimeToEnter))
                            if(RemaingTimeToJoin>0):
                                if:
                                    AllDungeonPosition:=AllDungeons_Positions[NumDungeon]
                                    StartPositionProp:=AllDungeonPosition[1]
                                    TranslationProp:=StartPositionProp.GetTransform().Translation
                                    RotationProp:=StartPositionProp.GetTransform().Rotation
                                then:
                                    if(Character := Agent.GetFortCharacter[]):                 
                                        if:
                                            Character.TeleportTo[TranslationProp,RotationProp]
                                        else:
                                            PrintWarning("TODO, safe teleport to other place if it fails")
                                else:
                                    #Print("IsToo Late!")   
                        DungeonUI.UpdateUI()  
            else:
                #Print("Fatal error, is not ready!")  
        else:
            #Print("Fatal error, trying to find tests manager or Dungeon UI")                          
            

    DungeonUI_CancelButtonPressed(Agent : agent, NumDungeon : int) : void =
        if(TestsManager := MaybeTestsManager?, DungeonUI:=TestsManager.AllDungeonUI[Agent]):
            RemoveAgentFromAllDungeons(Agent)           
            DungeonUI.UpdateUI()
       
    AddToDungeonAndStartCountDown(Agent : agent, NumDungeon : int)<suspends> : void =
        if(TestsManager := MaybeTestsManager?, AllDungeonUI:=TestsManager.AllDungeonUI):
            if(var AllAgents : []agent = AllDungeons_AgentsPlaying[NumDungeon]):
                if(AllAgents.Length<=0): #THERE IS NO PLAYERS, LETS PLAY IT
                    set AllAgents = AllAgents + array{Agent} #Add agent
                    if:
                        set AllDungeons_MaybeStartingAgent[NumDungeon]=option{Agent}
                        set AllDungeons_AgentsWaitingToJoin[NumDungeon]=AllAgents
                        set AllDungeons_CountToStartTime[NumDungeon] = 10
                        var TimeToJoin:  int = AllDungeons_CountToStartTime[NumDungeon]
                    then:
                        loop:
                            for(AgentWaiting:AllAgents):
                                if(ActionInfoUI := TestsManager.AllActionInfoUI[AgentWaiting]):
                                    ActionInfoUI.SetDungeon_TimeInfo("Starting in: {TimeToJoin}")
                                    ActionInfoUI.SetDungeon_ChestInfo("Find the chests and destroy the mutant bread")
                            ##Print("Countdown: {TimeToJoin}")
                            for(Player : GetPlayspace().GetPlayers(), OtherAgent := agent[Player], DungeonUI:=AllDungeonUI[OtherAgent], DungeonUI.Opened_NumDungeon=NumDungeon):
                                DungeonUI.UpdateUI()
                            Sleep(1.0)
                            set TimeToJoin = TimeToJoin - 1
                            if(set AllDungeons_CountToStartTime[NumDungeon]=TimeToJoin):
                            if(TimeToJoin<=0, MaybeStartingAgent:=AllDungeons_MaybeStartingAgent[NumDungeon], ValidStartingAgent := MaybeStartingAgent?):
                                StartDungeon(ValidStartingAgent, NumDungeon)
                                break
                            if(MaybeStartingAgent:=AllDungeons_MaybeStartingAgent[NumDungeon], ValidStartingAgent := MaybeStartingAgent?):
                            else:
                                break

    DungeonButton_Pressed(Agent : agent, NumDungeon : int, ButtonType : int)<suspends> : void =
        #Print("Detected press")
        if(TestsManager:=MaybeTestsManager?):
            if(ButtonType=ENTER_TYPE_BUTTON):
                TestsManager.Display_DungeonUI(Agent,NumDungeon)
            else if(ButtonType=EXIT_TYPE_BUTTON):
                QuickAgentFromDungeon(Agent,NumDungeon)
            else:
                IsAlreadyOpened:=TryOpenChest(Agent,NumDungeon,ButtonType)
                if(not IsAlreadyOpened?):
                    if(Player := player[Agent]):
                        #Print("ChestButton pressed!")
                        if(ButtonType=2):
                            if:
                                IsDefeated := AllDungeons_MinibossIsDefeated[NumDungeon]
                                IsDefeated?
                            then:
                                AllAwardsTuple:=DungeonDataBase.GetAndSaveChestAwards(Player,ButtonType,NumDungeon)
                                TestsManager.DisplayResourcersGather(Agent,AllAwardsTuple(0),AllAwardsTuple(1))
                                if(AllTreasuresOpened:=Dungeon_TresureRegistration[Agent].Length,AllTreasures:=AllDungeons_Buttons[NumDungeon].Length-2, AllTreasuresOpened=AllTreasures):
                                #Print("All treasures completed")
                                QuickAgentFromDungeon(Agent,NumDungeon)
                            else:
                                PrintWarning("TODO: Warn you must to defeat the miniboss before")
                        else:
                            AllAwardsTuple:=DungeonDataBase.GetAndSaveChestAwards(Player,ButtonType,NumDungeon)
                            TestsManager.DisplayResourcersGather(Agent,AllAwardsTuple(0),AllAwardsTuple(1))
                            if(AllTreasuresOpened:=Dungeon_TresureRegistration[Agent].Length,AllTreasures:=AllDungeons_Buttons[NumDungeon].Length-2, AllTreasuresOpened=AllTreasures):
                                #Print("All treasures completed")
                                QuickAgentFromDungeon(Agent,NumDungeon)

                PrintWarning("TO DO Dungeon reward get from pickaxe calculator")
        
    TryOpenChest(Agent : agent, NumDungeon : int, NumChest : int) : logic =
        var IsAlreadyOpened : logic = false 
        if(var TreasureRegistration : []int = Dungeon_TresureRegistration[Agent]):
            for(TreasureOpened : TreasureRegistration, TreasureOpened = NumChest):
                set IsAlreadyOpened = true
            if(IsAlreadyOpened?):
                PrintWarning("TODO Warn that this chest is already opened")
            else:
                set TreasureRegistration = TreasureRegistration + array{NumChest}
                if(set Dungeon_TresureRegistration[Agent] = TreasureRegistration):
                    set IsAlreadyOpened = false
        else:
            PrintFatalError("Treasure registration of this agent does not exists")
        return IsAlreadyOpened

    CreateMobsManagers() : void =
        for(Index:=0..10):
            NewMob := MakeMobManagerUI(Self, Index)
            set AllMobsManager = AllMobsManager + array{NewMob}
            ##Print("New mob manager set on {Index}")

    InitNPCSpawners() : void =
        for(Spawner : AllNPC_GroundMobSpawner):
            set AllNPC_GroundMobSpawner_IsBeingUsed = AllNPC_GroundMobSpawner_IsBeingUsed + array{-1}
            Spawner.SpawnedEvent.Subscribe(MobSpawnDetected)
        for(Spawner : AllNPC_DruidMobSpawner):
            set AllNPC_DruidMobSpawner_IsBeingUsed = AllNPC_DruidMobSpawner_IsBeingUsed + array{-1}
            Spawner.SpawnedEvent.Subscribe(MobSpawnDetected)
        for(Spawner : AllNPC_MiniBossMobSpawner):
            set AllNPC_MiniBossMobSpawner_IsBeingUsed = AllNPC_MiniBossMobSpawner_IsBeingUsed + array{-1}
            Spawner.SpawnedEvent.Subscribe(MobSpawnDetected)



    CleanNPCMemory(Assignation : int) : void =
        if(MobsManager := AllMobsManager[Assignation]):
            #Print("Cleaned assignation {Assignation}")
            MobsManager.SetUse(false)#>