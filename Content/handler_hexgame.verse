using { /Verse.org/Simulation }
using { /UnrealEngine.com/Temporary/UI }

handler_hexgame := class():
    # Positions used to access the respawnbikes this button controls.
    MaybeAgent : ?agent = false
    MaybeHexaMinigameManager : ?hexa_minigame_manager = false
    MaybePetsManager : ?pets_manager = false
    #MaybeGameManager : ?game_manager
    Direction : int = 0
    UISlot : int = -1
    IsCancel : logic = false
    #Submenu
    SubMenuOption : int = -1
    PurityAmountChange : int = 0

    #Area
    HandlerDirection(WidgetMessage : widget_message) : void =
        if(HexaMinigameManager := MaybeHexaMinigameManager?, Agent := agent[WidgetMessage.Player]):
            HexaMinigameManager.MoveAgentFocus(Agent, Direction)

    HandlerUIFromGeneralSelector(WidgetMessage : widget_message) : void =
        if(PetsManager := MaybePetsManager?, Agent := agent[WidgetMessage.Player]):
            spawn:
                PetsManager.Handler_FromGeneralSelectorMenu_To_Everything(WidgetMessage, IsCancel)

    #PetsMenu
    HandlerUIFromEverythingToMainMenu(WidgetMessage : widget_message) : void =
        if(PetsManager := MaybePetsManager?, Agent := agent[WidgetMessage.Player]):
            spawn:
                PetsManager.Handler_FromEverything_To_Menu(WidgetMessage, SubMenuOption)

    HandlerUIPetSubmenu(WidgetMessage : widget_message) : void =
        if(PetsManager := MaybePetsManager?, Agent := agent[WidgetMessage.Player]):
            spawn:
                PetsManager.Handler_FromMainMenu_To_Everything(WidgetMessage, SubMenuOption)

    HandlerUIGeneralSelector(WidgetMessage : widget_message) : void =
        #Print("Trying to display general selector")
        if(PetsManager := MaybePetsManager?):
            spawn:
                PetsManager.GeneralSelector_Display_Canvas(SubMenuOption, UISlot)

    HandlerUICreatePet_AddOrRemovePurity(WidgetMessage : widget_message) : void =
        if(PetsManager := MaybePetsManager?):
            PetsManager.CreatePet_AddOrRemovePurity(PurityAmountChange)