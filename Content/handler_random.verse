# fill this file with your verse-script
using { /Verse.org/Simulation }
using { /Fortnite.com/Devices }

handler_random := class():

    <#
    var Tutorial_0 : tuple(string,texture,string,int,float,int,float,?creative_prop,?button_device) = ("Breaking Bread I",Images.Pickaxe_Icon,"Pickaxe Slate Bread until you get Water",WATER,5.0,FLOUR,100.0,option{TutorialProp_0},option{TutorialButton_0})
            var Tutorial_1 : tuple(string,texture,string,int,float,int,float,?creative_prop,?button_device) = ("Fast travel I",Images.Speed_Icon,"Use the teleporter to travel to your Bakery",USETELEPORTER_TUTORIAL_TYPE,1.0,WATER,100.0,false,option{TutorialButton_1})
            var Tutorial_2 : tuple(string,texture,string,int,float,int,float,?creative_prop,?button_device) = ("Insert the resources",Images.Baker_Icon,"Insert the resources in the funnel",USEFUNNEL_TUTORIAL_TYPE,1.0,SAPPHIRES,1.0,false,option{TutorialButton_2})
            var Tutorial_3 : tuple(string,texture,string,int,float,int,float,?creative_prop,?button_device) = ("Using valves I",Images.Baker_Icon,"Adjust the valves until they are green or yellow",USEVALVES_TUTORIAL_TYPE,1.0,SAPPHIRES,1.0,false,option{TutorialButton_3})
            var Tutorial_3_5 : tuple(string,texture,string,int,float,int,float,?creative_prop,?button_device) = ("Baker Basics",Images.Baker_Icon,"Watch the Baker Tutorial",BAKERTUTORIAL_TUTORIAL_TYPE,1.0,SAPPHIRES,1.0,false,option{TutorialButton_35})
            var Tutorial_4 : tuple(string,texture,string,int,float,int,float,?creative_prop,?button_device) = ("Check the Baker info",Images.Baker_Icon,"Check Bakery info on the computer",CHECKINFO_TUTORIAL_TYPE,1.0,SAPPHIRES,1.0,false,option{TutorialButton_4})
            var Tutorial_5 : tuple(string,texture,string,int,float,int,float,?creative_prop,?button_device) = ("Expanding the business",Images.Pickaxe_Icon,"Pay the fee to unlock the next area",UNLOCKAREA_TUTORIAL_TYPE,1.0,SAPPHIRES,1.0,false,option{TutorialButton_5})
            var Tutorial_6 : tuple(string,texture,string,int,float,int,float,?creative_prop,?button_device) = ("Breaking Bread II",Images.Pickaxe_Icon,"Pickaxe Slate Bread until you get Water",WATER,15.0,WATER,250.0,option{TutorialProp_6},option{TutorialButton_6})
            var Tutorial_7 : tuple(string,texture,string,int,float,int,float,?creative_prop,?button_device) = ("Breaking Bread III",Images.Pickaxe_Icon,"Pickaxe Slate Bread until you get Flour",FLOUR,15.0,FLOUR,250.0,option{TutorialProp_7},option{TutorialButton_7})
            var Tutorial_8 : tuple(string,texture,string,int,float,int,float,?creative_prop,?button_device) = ("Baking Bread I",Images.Bread_Icon,"Bake Bread and send it using the drone",B_TYPE,1.0,SAPPHIRES,5.0,false,option{TutorialButton_8})
            var Tutorial_9 : tuple(string,texture,string,int,float,int,float,?creative_prop,?button_device) = ("Taking a break",Images.Bread_Icon,"Visit the rest room",USEAFKZONE_TUTORIAL_TYPE,1.0,SAPPHIRES,1.0,option{TutorialProp_9},option{TutorialButton_9})
            var Tutorial_10 : tuple(string,texture,string,int,float,int,float,?creative_prop,?button_device) = ("Breaking Harder I",Images.Pickaxe_Icon,"Improve your Pickaxe Damage",IMPROVEPICKAXE_TUTORIAL_TYPE,1.0,SAPPHIRES,5.0,false,option{TutorialButton_10})
            var Tutorial_11 : tuple(string,texture,string,int,float,int,float,?creative_prop,?button_device) = ("Upgrade your backpack",Images.Pickaxe_Icon,"Upgrade your backpack in the store",IMPROVEBACKPACK_TUTORIAL_TYPE,1.0,SAPPHIRES,5.0,false,option{TutorialButton_11})
            var Tutorial_12 : tuple(string,texture,string,int,float,int,float,?creative_prop,?button_device) = ("Upgrade your deposits",Images.Pickaxe_Icon,"Upgrade your Bakery deposits",IMPROVEMAINDEPOSITS_TUTORIAL_TYPE,1.0,SAPPHIRES,5.0,false,option{TutorialButton_12})
            var Tutorial_13 : tuple(string,texture,string,int,float,int,float,?creative_prop,?button_device) = ("Upgrade your Bakery",Images.Bread_Icon,"Upgrade your Bread Bakery production",IMPROVEBAKERY_TUTORIAL_TYPE,1.0,SAPPHIRES,5.0,false,option{TutorialButton_13})
            var Tutorial_14 : tuple(string,texture,string,int,float,int,float,?creative_prop,?button_device) = ("Baking Bread II",Images.Bread_Icon,"Bake more Bread and send it using the drone",B_TYPE,10.0,SAPPHIRES,10.0,false,option{TutorialButton_14})
            var Tutorial_15 : tuple(string,texture,string,int,float,int,float,?creative_prop,?button_device) = ("Breaking Harder II",Images.Pickaxe_Icon,"Upgrade your Pickaxe perks in the store",IMPROVEPICKPERK_TUTORIAL_TYPE,1.0,SAPPHIRES,10.0,false,option{TutorialButton_15})
            var Tutorial_16 : tuple(string,texture,string,int,float,int,float,?creative_prop,?button_device) = ("Breaking Bread IV",Images.Pickaxe_Icon,"Break 5 Slate Breads",BREAKBREAD_TUTORIAL_TYPE,5.0,SAPPHIRES,10.0,option{TutorialProp_16},option{TutorialButton_16})
            var Tutorial_17 : tuple(string,texture,string,int,float,int,float,?creative_prop,?button_device) = ("Using valves II",Images.Baker_Icon,"Set Cheese Bread Baker valves to green color",OPENINGVALVES_TUTORIAL_TYPE,1.0,SAPPHIRES,10.0,false,option{TutorialButton_17})
            var Tutorial_18 : tuple(string,texture,string,int,float,int,float,?creative_prop,?button_device) = ("Using valves III",Images.Baker_Icon,"Close all valves on your Bread Baker",CLOSINGVALVES_TUTORIAL_TYPE,1.0,SAPPHIRES,10.0,false,option{TutorialButton_18})
            var Tutorial_19 : tuple(string,texture,string,int,float,int,float,?creative_prop,?button_device) = ("Using valves IV",Images.Baker_Icon,"Set ALL valves to half open (Yellow) on ALL Bakers",HALFVALVES_TUTORIAL_TYPE,1.0,SAPPHIRES,10.0,false,option{TutorialButton_19})
            var Tutorial_20 : tuple(string,texture,string,int,float,int,float,?creative_prop,?button_device) = ("Breaking Bread V",Images.Pickaxe_Icon,"Pickaxe Slate Bread until you get Cheese",CHEESE,500.0,COAL,5.0,option{TutorialProp_20},option{TutorialButton_20})
            var Tutorial_21 : tuple(string,texture,string,int,float,int,float,?creative_prop,?button_device) = ("Baking Bread III",Images.CheeseBread_Icon,"Bake Cheese Bread and send it using the drone",CB_TYPE,25.0,COAL,5.0,false,option{TutorialButton_21})
            var Tutorial_22 : tuple(string,texture,string,int,float,int,float,?creative_prop,?button_device) = ("Using valves V",Images.PimentoCheeseSandwich_Icon,"Set Pimento Cheese Sandwich Baker to open",PCSBOPENINGVALVES_TUTORIAL_TYPE,1.0,COAL,5.0,false,option{TutorialButton_22})
            var Tutorial_23 : tuple(string,texture,string,int,float,int,float,?creative_prop,?button_device) = ("Using Coal I",Images.Coal_Icon,"Introduce 5 Coal in the Coal Generator",USECOAL_TUTORIAL_TYPE,5.0,COAL,5.0,false,option{TutorialButton_23})
            var Tutorial_24 : tuple(string,texture,string,int,float,int,float,?creative_prop,?button_device) = ("Stay out of my territory",Images.Pickaxe_Icon,"Pay the fee to unlock the next area",UNLOCKAREA_TUTORIAL_TYPE,1.0,COAL,5.0,false,option{TutorialButton_24})
            var Tutorial_24_5 : tuple(string,texture,string,int,float,int,float,?creative_prop,?button_device) = ("Meet the Team",Images.Icon_Druid,"Interact with a Sourdough on the table to meet it",PETTUTORIAL_TUTORIAL_TYPE,1.0,MUTANTMATERIAL,25.0,false,option{TutorialButton_36})
            var Tutorial_25 : tuple(string,texture,string,int,float,int,float,?creative_prop,?button_device) = ("Purest Formula I",Images.Mutant_Material_Icon,"Create a Sourdough using Mutant Material in the Sourdough Machine",CREATEPET_TUTORIAL_TYPE,1.0,MUTANTMATERIAL,5.0,false,option{TutorialButton_25})
            var Tutorial_26 : tuple(string,texture,string,int,float,int,float,?creative_prop,?button_device) = ("Purest Formula II",Images.Coal_Icon,"Use Coal to improve a Sourdough Pack's stats",IMPROVEPET_TUTORIAL_TYPE,1.0,COAL,5.0,false,option{TutorialButton_26})
            var Tutorial_27 : tuple(string,texture,string,int,float,int,float,?creative_prop,?button_device) = ("Purest Formula III",Images.Icon_ADC,"Unlock a Sourdough Competition Challenge",UNLOCKHEXAGAME_TUTORIAL_TYPE,1.0,COAL,5.0,option{TutorialProp_27},option{TutorialButton_27})
            var Tutorial_28 : tuple(string,texture,string,int,float,int,float,?creative_prop,?button_device) = ("Fast travel II",Images.Speed_Icon,"Use Coal on the teleporter to travel to your Bakery",USEEXPENSIVETELEPORTER_TUTORIAL_TYPE,1.0,COAL,5.0,false,option{TutorialButton_28})
            var Tutorial_29 : tuple(string,texture,string,int,float,int,float,?creative_prop,?button_device) = ("Purest Formula IV",Images.Icon_ADC,"Play and win a Sourdough Competition Challenge",WINHEXAGAME_TUTORIAL_TYPE,1.0,COAL,5.0,false,option{TutorialButton_29})
            var Tutorial_30 : tuple(string,texture,string,int,float,int,float,?creative_prop,?button_device) = ("Complete an order",Images.PimentoCheeseSandwich_Icon,"Send Pimento Cheese Sandwiches using the drone",PCS_TYPE,150.0,COAL,5.0,false,option{TutorialButton_30})
            var Tutorial_32 : tuple(string,texture,string,int,float,int,float,?creative_prop,?button_device) = ("Weapon improve",Images.Sword_Icon,"Upgrade the Hammer in the store",WEAPONIMPROVE_TUTORIAL_TYPE,1.0,COAL,5.0,false,option{TutorialButton_32})
            var Tutorial_31 : tuple(string,texture,string,int,float,int,float,?creative_prop,?button_device) = ("The bad batch",Images.Sword_Icon,"Clear a dungeon and destroy the mutant bread",COMPLETEDUNGEON_TUTORIAL_TYPE,1.0,COAL,5.0,false,option{TutorialButton_31})
    #>
    # Positions used to access the respawnbikes this button controls.
    MaybeTeleporterManager : ?teleporter_manager = false
    MaybeVFXManager : ?vfx_and_accolades_manager =  false
    MaybeCollectibleManager : ?collectibles_manager =  false
    MaybeMinigamesManager : ?minigames_manager =  false
    MaybeQuestAndOrders : ?quests_and_orders_database = false
    MaybeGameManager : ?game_manager =  false
    MaybeProp : ?creative_prop = false 
    PropType : int = -1
    Index : int = -1
    AuxIndex : int = -1
    MiniGameQuestType : int  = -1
    Amount : float = -1.0
    Color : int = -1

    
    HandlerLoreInteraction(Agent : agent) : void =
        if(QuestAndOrders:=MaybeQuestAndOrders?):
            QuestAndOrders.LoreInteraction(Agent,Index)
    
    HandlerCheckMusic(Agent : agent) : void =
        Log("HandlerCheckMusic",LOG_LEVEL_DEBUG6)
        if(GameManager := MaybeGameManager?):
            Log("Checking agent music: {Index}",LOG_LEVEL_DEBUG6)
            GameManager.CheckZoneMusic(Agent,Index)
        else:
            PrintFatalError("GameManager not found")

    HandlerChange_Material(KeyFrame : int, IsReverse : logic) : void =
        if(VFXManager := MaybeVFXManager?, Prop:=MaybeProp?):
            VFXManager.Change_Material(Prop,KeyFrame,PropType)

    HandlerOpenTeleporterMenu(Agent : agent) : void =
        if(GameManager:=MaybeGameManager?):
            MaybeBakery:=GameManager.GetAgentBakery(Agent)
            if(BakeryMachinesManager := MaybeBakery?):
                BakeryMachinesManager.Display_TeleporterMenu(Agent,Index)
        else:
            PrintFatalError("On HandlerOpenTeleporterMenu Bakery not valid")
            
    HandlerTeleporterButton(Agent : agent) : void =
        if(TeleporterManager:=MaybeTeleporterManager?):
            TeleporterManager.TryToTeleportToBakery(Index, Agent)

    HandlerCollectibleSwitch(Agent : agent) : void =
        if(CollectibleManager := MaybeCollectibleManager?):
            CollectibleManager.TryRegisterSecret(Agent, Index)

    HandlerCollectibleHexaSwitch(Agent : agent) : void =
        if(CollectibleManager := MaybeCollectibleManager?):
            CollectibleManager.TryRegisterHexaCollectibleSecret(Agent, Index)

    HandlerArcheryDamagedGame(Agent : agent) : void =
        if(MinigamesManager := MaybeMinigamesManager?):
            MinigamesManager.ArcheryDamagedEvent(Agent, Index)

    HandlerMinigamesManagerArcheryButton(Agent : agent) : void =
        if(MinigamesManager := MaybeMinigamesManager?):
            MinigamesManager.ArcheryButtonTriggered(Agent, Index)

    HandlerSkilledGameCompleted(Agent : agent) : void =
        if(MinigamesManager := MaybeMinigamesManager?):
            MinigamesManager.TryAwardSkilledMiniGame(Agent, MiniGameQuestType, Index, AuxIndex, Amount)

    HandlerDisplaySkilledGame(Agent : agent) : void =
        if(MinigamesManager := MaybeMinigamesManager?):
            MinigamesManager.DisplaySkilledMiniGame(Agent,Index)

    HandlerWeakPointEvent(Agent : agent) : void =
        if(GameManager := MaybeGameManager?):
            Log("WeakPointEvent detected", LOG_LEVEL_UIs)
            GameManager.WeakPointEvent(Agent, Index)

    HandlerSimonManipulator(Agent : agent) : void =
        if(MinigamesManager:=MaybeMinigamesManager?):
            MinigamesManager.SimonPressed(Agent,Index,Color)