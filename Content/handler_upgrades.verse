# fill this file with your verse-script
using { /Verse.org/Simulation }
using { /Fortnite.com/Devices }
using { /UnrealEngine.com/Temporary/UI }

handler_upgrades := class():
    # Positions used to access the respawnbikes this button controls.
    MaybeAgent : ?agent = false
    MaybeBakeryMachinesManager : ?bakery_machines_manager = false
    MaybeGameManager : ?game_manager = false
    #MaybeGameManager : ?game_manager
    Index : int = 0
    Row : int = 0
    Level : int = 0
    BakerType : int = 0
    Resource : int = 0
    Perk : int = 0
    CraftMode : logic = false

    #Area
    HandlerTryUnlockArea(Agent : agent) : void =
        if(GameManager := MaybeGameManager?):
            GameManager.TryUnlockArea(Agent, Index)

    #Valves
    HandlerTryChangeValveState(Agent : agent) : void =
        if(BakeryMachinesManager := MaybeBakeryMachinesManager?):
            BakeryMachinesManager.Change_ValveState(Agent, Index)

    HandlerTryChangeSimplifiedValveState(Agent : agent) : void =
        if(BakeryMachinesManager := MaybeBakeryMachinesManager?):
            BakeryMachinesManager.Change_SimplifiedValveState(Agent, BakerType)

    HandlerPickaxeUpgrade(WidgetMessage : widget_message) : void =
        if(BakeryMachinesManager := MaybeBakeryMachinesManager?):
            BakeryMachinesManager.TryUpgradePickaxe_Strength(Row)

    HandlerPickaxeUpgradeEfficiency(WidgetMessage : widget_message) : void =
        if(BakeryMachinesManager := MaybeBakeryMachinesManager?):
            BakeryMachinesManager.TryUpgradePickaxe_Efficiency(Row)

    HandlerUpgradeTier(WidgetMessage : widget_message) : void =
        if(BakeryMachinesManager := MaybeBakeryMachinesManager?):
            BakeryMachinesManager.TryUpgradeBaker_Production(Row, BakerType)
        
    HandlerUpgradeEfficiency(WidgetMessage : widget_message) : void =
        if(BakeryMachinesManager := MaybeBakeryMachinesManager?):
            BakeryMachinesManager.TryUpgradeBaker_Efficiency(Row, BakerType)

    HandlerUpgradeSpeed(WidgetMessage : widget_message) : void =
        if(BakeryMachinesManager := MaybeBakeryMachinesManager?):
            BakeryMachinesManager.TryUpgradeBaker_Speed(Row, BakerType)

    #Deposit
    HandlerDepositUpgrade(WidgetMessage : widget_message) : void =
        if(BakeryMachinesManager := MaybeBakeryMachinesManager?):
            BakeryMachinesManager.TryUpgradeDeposit_Capacity(Row)

    HandlerDepositPasiveUpgrade(WidgetMessage : widget_message) : void =
        if(BakeryMachinesManager := MaybeBakeryMachinesManager?):
            BakeryMachinesManager.TryUpgradeDeposit_PasiveIncome(Row)

    #Drones
    HandlerDroneSpeedUpgrade(WidgetMessage : widget_message) : void =
        if(BakeryMachinesManager := MaybeBakeryMachinesManager?):
            BakeryMachinesManager.TryUpgradeDroneSpeed()

    HandlerDroneCapacityUpgrade(WidgetMessage : widget_message) : void =
        if(BakeryMachinesManager := MaybeBakeryMachinesManager?):
             #BakeryMachinesManager.TryUpgradeDroneCapacity()

    HandlerSalesEfficiencyUpgrade(WidgetMessage : widget_message) : void =
        if(BakeryMachinesManager := MaybeBakeryMachinesManager?):
            BakeryMachinesManager.TrySalesEfficiencyUpgrade(BakerType)

    #Backpack
    HandlerBackpackUpgrade(WidgetMessage : widget_message) : void =
        if(BakeryMachinesManager := MaybeBakeryMachinesManager?):
            BakeryMachinesManager.TryBackpackUpgrade(Resource)

    #Pickaxe Perks
    HandlerPickaxePerkUpgrade(WidgetMessage : widget_message) : void =
        if(BakeryMachinesManager := MaybeBakeryMachinesManager?):
            BakeryMachinesManager.TryPickaxePerkUpgrade(Perk)