using { /Verse.org/Simulation }
using { /UnrealEngine.com/Temporary/UI }
using { /UnrealEngine.com/Temporary/SpatialMath }
using { /Fortnite.com/Characters }

handler_weaponsandmobs := class():

    MaybeTestsManager : ?tests_manager = false
    MaybeGeneralMobsManager : ?general_mobs_manager = false
    MaybeWeaponsUpgradesUI : ?weapons_upgrades_ui = false
    MaybeTokensUIManager : ?tokens_ui = false
    WeaponPerk : int = 0
    Position : transform = transform{}
    MobType : int = GROUND_MOB
    MobLevel : int = 0
    MobTier : int = 0

    SpawnerIndex : int = 0

    DungeonToken_Perk : int = 0

    DungeonNumber : int = 0
    DungeonButtonType : int = 0

    HandlerPropManipulator(Agent : agent) : void =
        if(TestsManager := MaybeTestsManager?):
           TestsManager.SpawnRandomEncounter(Agent,MobLevel,Position)

    #Area
    HandlerImproveWeapon(Message : widget_message) : void =
        if(WeaponsUpgradesUI := MaybeWeaponsUpgradesUI?):
            WeaponsUpgradesUI.TryImprovePerk(WeaponPerk)

    HandlerDungeonMobSpawn(MobAgent : agent) : void =
        if(GeneralMobsManager := MaybeGeneralMobsManager?):
            GeneralMobsManager.DungeonMobSpawnDetected(MobAgent,SpawnerIndex, MobType)

    HandlerDungeonButton(Agent : agent) : void =
        if(GeneralMobsManager := MaybeGeneralMobsManager?):
            spawn:
                GeneralMobsManager.DungeonButton_Pressed(Agent, DungeonNumber, DungeonButtonType)

    #Tokens
    HandlerDTTokensUI(Message : widget_message) : void =
        if(TokensUIManager := MaybeTokensUIManager?):
            TokensUIManager.TryImproveDTPerk(DungeonToken_Perk)

    HandlerHTTokensUI(Message : widget_message) : void =
        if(TokensUIManager := MaybeTokensUIManager?):
            TokensUIManager.TryImproveHTPerk(DungeonToken_Perk)


                