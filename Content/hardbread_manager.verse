using { /UnrealEngine.com/Temporary/SpatialMath }
using { /Fortnite.com/Devices }
using { /Verse.org/Simulation }
using { /UnrealEngine.com/Temporary/Diagnostics }
using { /Fortnite.com/Devices/CreativeAnimation }
using { /Fortnite.com/Devices/CreativeAnimation/InterpolationTypes }
using { /Verse.org/Random }
using { /Fortnite.com/Characters }

# See https://dev.epicgames.com/documentation/en-us/uefn/create-your-own-device-in-verse for how to create a verse device.

# A Verse-authored creative device that can be placed in a level
hardbread_manager := class(creative_device):

    <#
    INDESTRUCTIBLE : int = 0
    #INESTABLE : int = -1

    @editable
    IsInestable : logic = false
    var InestabilityRunningTime : int = 0
    @editable
    Level : int = 0
    var HBType : int = 0
    var TotalTimeWithoutDamage : int = -1

    var GameManager : game_manager = game_manager{}
    var AllPlayersNumHits_ForWeakPoint : [agent]int = map{}
    var AllPlayersNumHits : [agent]int = map{}

    var DamageAcumulated : float = 0.0
    var ConsecutiveHits : int = 0
    var WeakPoint : int = -1

    var MaxHealth : float = 1000.0
    var TotalHealth : float = 1000.0
    var DamageToWeakPoint : float = 0.0
    var IsDamageAnimationPlaying : logic = false 
    var IsDeathAnimationPlaying : logic = false 

    @editable
    BreadRoot : creative_prop = creative_prop{} 

    @editable
    BreadPiece0 : creative_prop = creative_prop{}
    @editable
    BreadPiece1 : creative_prop = creative_prop{}
    @editable
    BreadPiece2 : creative_prop = creative_prop{}
    @editable
    BreadPiece3 : creative_prop = creative_prop{}
    @editable
    BreadPiece4 : creative_prop = creative_prop{}
    @editable
    BreadPiece5 : creative_prop = creative_prop{}
    @editable
    BreadPiece6 : creative_prop = creative_prop{}
    @editable
    BreadPiece7 : creative_prop = creative_prop{}

    @editable
    Manipulator0 : prop_manipulator_device = prop_manipulator_device{}
    @editable
    Manipulator1 : prop_manipulator_device = prop_manipulator_device{}
    @editable
    Manipulator2 : prop_manipulator_device = prop_manipulator_device{}
    @editable
    Manipulator3 : prop_manipulator_device = prop_manipulator_device{}
    @editable
    Manipulator4 : prop_manipulator_device = prop_manipulator_device{}
    @editable
    Manipulator5 : prop_manipulator_device = prop_manipulator_device{}
    @editable
    Manipulator6 : prop_manipulator_device = prop_manipulator_device{}
    @editable
    Manipulator7 : prop_manipulator_device = prop_manipulator_device{}

    #Death settings
    DecrementOnZ : float = 64.0
    var DelayToResurrect : float = 5.0
    var OriginalPosition : transform = transform{}
    var AllHPieces : []creative_prop = array{}

    # Runs when the device is started in a running game
    OnBegin<override>()<suspends>:void=
        InitSubscribes()
        set AllHPieces = array{BreadPiece0,BreadPiece1,BreadPiece2,BreadPiece3,BreadPiece4,BreadPiece5,BreadPiece6,BreadPiece7}
        Sleep(1.0)
        ResetSetType()
        SetMaterial()
        Sleep(2.0)
        SavePosition()
        GenerateHealth()

    SpawnHealthChecker()<suspends> : void =
        MaxTimeWithoutDamage := 30
        if(TotalTimeWithoutDamage=-1):
            set TotalTimeWithoutDamage=MaxTimeWithoutDamage
            loop:
                Sleep(1.0)
                set TotalTimeWithoutDamage=TotalTimeWithoutDamage-1
                if(TotalTimeWithoutDamage<=0):
                    ResetStats()
                    SetMaterial()
                    set TotalTimeWithoutDamage = -1
                    break
        else:
            set TotalTimeWithoutDamage=MaxTimeWithoutDamage


    ResetSetType() : void =
        var PosibleTypes : []int = array{}
        if(IsInestable?):
            set PosibleTypes = array{HB_TYPE4} #Inestable
        else:
            if(Level<5): 
                if(Level<2): 
                    set PosibleTypes = array{HB_TYPE0,HB_TYPE1} #Water. Flour.
                else:
                    set PosibleTypes = array{HB_TYPE0,HB_TYPE1,HB_TYPE2} #Water. Flour. Cheese
            else:
                set PosibleTypes = array{HB_TYPE0,HB_TYPE1,HB_TYPE2,HB_TYPE3} #Water. Flour. Cheese. Pimentos. 

        RandomType := GetRandomInt(0,PosibleTypes.Length-1)
        if(Value := PosibleTypes[RandomType], set HBType = Value):
            if(Level=INDESTRUCTIBLE):
                set HBType=HB_BASETYPE
            
    SetMaterial() : void =
        if(Level<>INDESTRUCTIBLE):
            if(HBType=HB_TYPE0):
                for(Piece : AllHPieces):
                    Piece.SetMaterial(Mat_HB_Water, ?Index:=0)
            else if(HBType=HB_TYPE1):
                for(Piece : AllHPieces):
                    Piece.SetMaterial(Mat_HB_Flour, ?Index:=0)
            else if(HBType=HB_TYPE2):
                for(Piece : AllHPieces):
                    Piece.SetMaterial(Mat_HB_Cheese, ?Index:=0)
            else if(HBType=HB_TYPE3):
                for(Piece : AllHPieces):
                    Piece.SetMaterial(Mat_HB_Pimentos, ?Index:=0)
            else if(HBType=HB_TYPE4):
                for(Piece : AllHPieces):
                    Piece.SetMaterial(Mat_HB_MutantMaterial, ?Index:=0)

    SetWeakPointMaterial() : void =
        if(PieceSelected := AllHPieces[WeakPoint]):
            if(HBType=HB_TYPE0):
                PieceSelected.SetMaterial(Mat_HB_WaterWeak, ?Index:=0)
            else if(HBType=HB_TYPE1):
                PieceSelected.SetMaterial(Mat_HB_FlourWeak, ?Index:=0)
            else if(HBType=HB_TYPE2):
                PieceSelected.SetMaterial(Mat_HB_CheeseWeak, ?Index:=0)
            else if(HBType=HB_TYPE3):
                PieceSelected.SetMaterial(Mat_HB_PimentosWeak, ?Index:=0)
            else if(HBType=HB_TYPE4):
                PieceSelected.SetMaterial(Mat_HB_MutantMaterialWeak, ?Index:=0)

    SetGameManager(InGameManager : game_manager) : void =
        set GameManager = InGameManager

    SavePosition() : void =
        set OriginalPosition = BreadRoot.GetTransform()

    GenerateHealth() : void =
        set MaxHealth = MainFormula_GenerateHardBreadHealth(Level,HBType)
        set TotalHealth = MaxHealth
        set DamageToWeakPoint = TotalHealth*PercentageToWeakPoint

    ResetStats() : void =
        set AllPlayersNumHits_ForWeakPoint = map{}
        set AllPlayersNumHits = map{}
        set DamageAcumulated = 0.0
        set ConsecutiveHits = 0
        set WeakPoint = -1
        set TotalHealth = MaxHealth

    #Damage num piece
    DamagedPiece0(Agent : agent) : void =
        #Print("Piece 0 damaged")
        Manipulator0.RestockResources()
        DamageDealt(Agent,0)
    DamagedPiece1(Agent : agent) : void =
        #Print("Piece 1 damaged")
        Manipulator1.RestockResources()
        DamageDealt(Agent,1)
    DamagedPiece2(Agent : agent) : void =
        #Print("Piece 2 damaged")
        Manipulator2.RestockResources()
        DamageDealt(Agent,2)
    DamagedPiece3(Agent : agent) : void =
        #Print("Piece 3 damaged")
        Manipulator3.RestockResources()
        DamageDealt(Agent,3)
    DamagedPiece4(Agent : agent) : void =
        #Print("Piece 4 damaged")
        Manipulator4.RestockResources()
        DamageDealt(Agent,4)
    DamagedPiece5(Agent : agent) : void =
        Manipulator5.RestockResources()
        #Print("Piece 5 damaged")
        DamageDealt(Agent,5)
    DamagedPiece6(Agent : agent) : void =
        Manipulator6.RestockResources()
        #Print("Piece 6 damaged")
        DamageDealt(Agent,6)
    DamagedPiece7(Agent : agent) : void =
        Manipulator7.RestockResources()
        #Print("Piece 7 damaged")
        DamageDealt(Agent,7)

    CalculeDamageDealtAndResourcers(Agent : agent, NumPiece : int) : float =
        var IsWeakPoint : logic = false
        var PickaxeDamage : float = 0.0
        #Log("ConsecutiveHits: {DamageToWeakPoint}/{DamageAcumulated}",LOG_LEVEL_UIs)
        #Log("WeakPoint is {WeakPoint}", LOG_LEVEL_UIs)
        #Check if it is weak point
        if(NumPiece=WeakPoint):
            Log("Is Weak point {WeakPoint}", LOG_LEVEL_UIs)
            GameManager.VFXManager.Spawn_HardBreadCriticalStrike(Agent,HBType)
            set IsWeakPoint = true
            SetMaterial()
            set WeakPoint = -1
        else:
            set IsWeakPoint = false
        #Damage to the cube
        MaybeBakery:=GameManager.GetAgentBakery(Agent)
        if(ValidBakery:=MaybeBakery?):
            var ResultTuple : tuple(float,[]int,[]float,[]logic,logic) = (0.0,array{0},array{0.0},array{false},false)
            var NumHitsAccumulated_ForWeakPoint : int = 0
            if(Level<>0):
                #Num weak point Hits
                if(Value:=AllPlayersNumHits_ForWeakPoint[Agent]):
                    set NumHitsAccumulated_ForWeakPoint = Value
                set NumHitsAccumulated_ForWeakPoint = NumHitsAccumulated_ForWeakPoint + 1
                if(set AllPlayersNumHits_ForWeakPoint[Agent] = NumHitsAccumulated_ForWeakPoint):
                    Log("NumHitsAccumulated_ForWeakPoint accumulated {NumHitsAccumulated_ForWeakPoint}", LOG_LEVEL_UIs)

                #Num total hits
                var NumHitsAccumulated : int =0
                if(Value:=AllPlayersNumHits[Agent]):
                    set NumHitsAccumulated = Value
                set NumHitsAccumulated = NumHitsAccumulated + 1
                if(set AllPlayersNumHits[Agent] = NumHitsAccumulated):
                    Log("NumHitsAccumulated accumulated {NumHitsAccumulated}", LOG_LEVEL_UIs)
                #Get resourcers
                set ResultTuple = ValidBakery.GetDamage_And_Resourcers(Agent,NumHitsAccumulated,HBType,Level,IsWeakPoint,TotalHealth)
            if(Level=0):
                set ResultTuple = ValidBakery.GetDamage_And_Resourcers(Agent,0,HBType,Level,IsWeakPoint,TotalHealth)
            set PickaxeDamage = ResultTuple(0)
            AllResourcersType := ResultTuple(1)
            AllResourcersAmmount := ResultTuple(2)
            #Log("PickaxeDamage: {PickaxeDamage}",LOG_LEVEL_UIs)
            #Log("AllResourcersType.Length: {AllResourcersType.Length}",LOG_LEVEL_UIs)
            #Log("AllResourcersAmmount.Length: {AllResourcersAmmount.Length}",LOG_LEVEL_UIs)
            AllLimitsReached := ResultTuple(3)
            IsCritical := ResultTuple(4)
            #Display resourcers
            ValidBakery.DisplayDamageUI(Agent, AllResourcersAmmount,AllResourcersType, AllLimitsReached,IsCritical)
            if(DamageToWeakPoint <= DamageAcumulated or NumHitsAccumulated_ForWeakPoint > GameManager.MaxNumHitsToWeakPoint, WeakPoint = -1):
                set DamageAcumulated = 0.0
                if(set AllPlayersNumHits_ForWeakPoint[Agent]=0):
                RandomInt:=GetRandomInt(0,7)
                #Log("Weak point seleted: {WeakPoint}",LOG_LEVEL_UIs)
                GameManager.VFXManager.PlayWeakPointExposed(Agent)
                set WeakPoint = RandomInt
                SetWeakPointMaterial()
            else:
                set DamageAcumulated = DamageAcumulated + PickaxeDamage
        return PickaxeDamage

    #Damage
    DamageDealt(Agent: agent, NumPiece : int) : void =
        spawn:
            SpawnHealthChecker()
        MaybeBakery:=GameManager.GetAgentBakery(Agent)
        if(ValidBakery:=MaybeBakery?):
            if(TotalHealth>0.0, HealthUI := ValidBakery.MaybeEntityHealthUI?): #If it is death already, don´t do anything
                TotalDamage := CalculeDamageDealtAndResourcers(Agent, NumPiece)   
                if(Level<>INDESTRUCTIBLE, not IsInestable?):
                    set TotalHealth = TotalHealth - TotalDamage
                #Log("TotalDamage: {TotalDamage}, TotalHealth: {TotalHealth}",LOG_LEVEL_UIs)
                if(TotalHealth<=0.0):
                    set TotalHealth = 0.0   
                    HealthUI.SetInfo(0.0,MaxHealth,HBType,Level)       
                    #Play death effect
                    if(not IsDeathAnimationPlaying?):
                        set IsDeathAnimationPlaying = true
                        spawn:
                            AnimateDeathEffect(Agent,0.3)
                else:
                    HealthUI.SetInfo(TotalHealth,MaxHealth,HBType,Level)
                    #Play damage effect
                    if(not IsDamageAnimationPlaying?, TotalDamage > (TotalHealth*0.01)):
                        set IsDamageAnimationPlaying = true
                        spawn:
                            AnimateDamagedEffect()

    DestroyBread(EliminatorAgent : agent)<suspends> : void =
        AMobWillBeSpawned:=CheckRandomEncounter(EliminatorAgent)
        if(AMobWillBeSpawned?):
            Sleep(RandomMobMaxLifeTime+2.0)
        else:
            Sleep(DelayToResurrect)
        InitResurect()

    CheckRandomEncounter(EliminatorAgent : agent) : logic =
        var TotalNumHits : int = 0
        var AMobWillBeSpawned : logic = false
        for(PlayersNumHits : AllPlayersNumHits):
            set TotalNumHits = TotalNumHits + PlayersNumHits
        if(TotalNumHits>=15):
            set AMobWillBeSpawned = GameManager.GeneralMobsManager.RollRandomPlantEncounter(EliminatorAgent, OriginalPosition, false, Level, TotalNumHits)
        return AMobWillBeSpawned

    InitSubscribes<private>() : void =
        Manipulator0.HarvestingEvent.Subscribe(DamagedPiece0)
        Manipulator1.HarvestingEvent.Subscribe(DamagedPiece1)
        Manipulator2.HarvestingEvent.Subscribe(DamagedPiece2)
        Manipulator3.HarvestingEvent.Subscribe(DamagedPiece3)
        Manipulator4.HarvestingEvent.Subscribe(DamagedPiece4)
        Manipulator5.HarvestingEvent.Subscribe(DamagedPiece5)
        Manipulator6.HarvestingEvent.Subscribe(DamagedPiece6)
        Manipulator7.HarvestingEvent.Subscribe(DamagedPiece7)


    InitResurect<private>()<suspends> : void =
        MoveTime : float = 2.0
        UpCrementOnZ := DecrementOnZ * 4
        ResetSetType()
        spawn:
            BreadRoot.MoveTo(OriginalPosition,MoveTime)
        Sleep(MoveTime/2.0)
        SetMaterial()
        CheckForOverlapedPlayers()
        Sleep(MoveTime/2.0)
        CheckForOverlapedPlayers()
        ResetStats()
        

    CheckForOverlapedPlayers() : void =
        for(Player : GetPlayspace().GetPlayers(), Character := Player.GetFortCharacter[],168.0>Distance(Character.GetTransform().Translation,BreadRoot.GetTransform().Translation)):
            NewTransform := vector3{X:=Character.GetTransform().Translation.X,Y:=Character.GetTransform().Translation.Y,Z:=Character.GetTransform().Translation.Z+128.0}
            if(Character.TeleportTo[NewTransform,Character.GetTransform().Rotation]):
            else:
                TeleportFailedAttempt(option{Character}, false, false, Character.GetTransform(), 64.0)

    AnimateDeathEffect<public>(Agent : agent, Delay : float)<suspends> : void =
        spawn:
            GameManager.VFXManager.Spawn_HardBreadDeath(Agent,BreadRoot,HBType)
        SetMaterial()
        EmptyTranslation := vector3{X:=0.0,Y:=0.0,Z:=0.0}
        X:=1.0+0.2
        
        #Roll is X: Red Axis OF THE OBJECT
        #Pitch is Y: Green Axis OF THE OBJECT
        #Yaw is Z: Blue axis OF THE OBJECT

        Rot1_RotationRoll : float = 20.0
        Rot2_RotationRoll : float = -40.0
        Rot3_RotationRoll : float = 40.0
        Rot4_RotationRoll : float = -20.0

        RotationApplied1 := MakeRotationFromYawPitchRollDegrees(0.0, 0.0,Rot1_RotationRoll)
        RotationApplied2 := MakeRotationFromYawPitchRollDegrees(0.0, 0.0,Rot2_RotationRoll)
        RotationApplied3 := MakeRotationFromYawPitchRollDegrees(0.0, 0.0,Rot3_RotationRoll)
        RotationApplied4 := MakeRotationFromYawPitchRollDegrees(0.0, 0.0,Rot4_RotationRoll)
        

        MovementKeyFrame1 : keyframe_delta = keyframe_delta #Key Frame 1 to base position

        {
            DeltaLocation := vector3{X:=0.0,Y:=0.0,Z:=-DecrementOnZ}
            DeltaRotation := RotationApplied1
            DeltaScale := vector3{X:=0.5,Y:=0.5,Z:=0.5}
            Time := 0.25
            Interpolation := EaseIn
        }

        MovementKeyFrame2 : keyframe_delta = keyframe_delta #Key to BouncePoint with Bounce Scale
        {
            DeltaLocation := vector3{X:=0.0,Y:=0.0,Z:=-DecrementOnZ}
            DeltaRotation := RotationApplied2
            DeltaScale := vector3{X:=0.5,Y:=0.5,Z:=0.5}
            Time := 0.25
            Interpolation := EaseIn
        }

        MovementKeyFrame3 : keyframe_delta = keyframe_delta #Key to BouncePoint with Bounce Scale
        {
            DeltaLocation := vector3{X:=0.0,Y:=0.0,Z:=-DecrementOnZ}
            DeltaRotation := RotationApplied3
            DeltaScale := vector3{X:=0.5,Y:=0.5,Z:=0.5}
            Time := 0.25
            Interpolation := EaseIn
        }

        MovementKeyFrame4 : keyframe_delta = keyframe_delta #Key to BouncePoint with Bounce Scale
        {
            DeltaLocation := vector3{X:=0.0,Y:=0.0,Z:=-DecrementOnZ}
            DeltaRotation := RotationApplied4
            DeltaScale := vector3{X:=0.5,Y:=0.5,Z:=0.5}
            Time := 0.25
            Interpolation := EaseIn
        }

        if(PropController := BreadRoot.GetAnimationController[]):
            ##Print("Playing death animation")
            # Make an array of key frames
            FortVaderKeyFrames :  []keyframe_delta = array { MovementKeyFrame1 , MovementKeyFrame2,  MovementKeyFrame3, MovementKeyFrame4}
            # Set the animation to play
            PropController.SetAnimation(FortVaderKeyFrames, ?Mode:=animation_mode.OneShot)
            # Play Animation
            #Sleep(Delay)
            if(IsDamageAnimationPlaying?):
                Sleep(Delay)
            if(not IsDamageAnimationPlaying?):
                PropController.Play()
                loop:
                    State := PropController.GetState()
                    if(State = animation_controller_state.Stopped):
                        set IsDeathAnimationPlaying = false 
                        spawn:
                            DestroyBread(Agent)
                        break
                    Sleep(0.04)

    AnimateDamagedEffect<public>()<suspends> : void =
        EmptyTranslation := vector3{X:=0.0,Y:=0.0,Z:=0.0}
        X:=1.0+0.2
        
        #Roll is X: Red Axis OF THE OBJECT
        #Pitch is Y: Green Axis OF THE OBJECT
        #Yaw is Z: Blue axis OF THE OBJECT

        Rot1_RotationRoll : float = 2.0
        Rot2_RotationRoll : float = -4.0
        Rot3_RotationRoll : float = 2.0

        NormalScale : vector3 = vector3{X:=1.0,Y:=1.0,Z:=1.0}

        RotationApplied1 := MakeRotationFromYawPitchRollDegrees(0.0, 0.0,Rot1_RotationRoll)
        RotationApplied2 := MakeRotationFromYawPitchRollDegrees(0.0, 0.0,Rot2_RotationRoll)
        RotationApplied3 := MakeRotationFromYawPitchRollDegrees(0.0, 0.0,Rot3_RotationRoll)
        

        MovementKeyFrame1 : keyframe_delta = keyframe_delta #Key Frame 1 to base position

        {
            DeltaLocation := EmptyTranslation
            DeltaRotation := RotationApplied1
            DeltaScale := NormalScale
            Time := 0.1
            Interpolation := EaseIn
        }

        MovementKeyFrame2 : keyframe_delta = keyframe_delta #Key to BouncePoint with Bounce Scale
        {
            DeltaLocation := EmptyTranslation 
            DeltaRotation := RotationApplied2
            DeltaScale := NormalScale
            Time := 0.1
            Interpolation := EaseIn
        }

        MovementKeyFrame3 : keyframe_delta = keyframe_delta #Key Frame 3 to go to the Destiny
        {
            DeltaLocation := EmptyTranslation 
            DeltaRotation := RotationApplied3
            DeltaScale := NormalScale
            Time := 0.1
            Interpolation := EaseIn
        }

        if(PropController := BreadRoot.GetAnimationController[]):
            ##Print("Playing damage animation")
            # Make an array of key frames
            FortVaderKeyFrames :  []keyframe_delta = array { MovementKeyFrame1 , MovementKeyFrame2,  MovementKeyFrame3}
            # Set the animation to play
            PropController.SetAnimation(FortVaderKeyFrames, ?Mode:=animation_mode.OneShot)
            # Play Animation
            PropController.Play()
            loop:
                State := PropController.GetState()
                if(State = animation_controller_state.Stopped):
                    set IsDamageAnimationPlaying = false 
                    break
                Sleep(0.04)


 
#>