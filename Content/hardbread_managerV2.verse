using { /UnrealEngine.com/Temporary/SpatialMath }
using { /Fortnite.com/Devices }
using { /Verse.org/Simulation }
using { /UnrealEngine.com/Temporary/Diagnostics }
using { /Fortnite.com/Devices/CreativeAnimation }
using { /Fortnite.com/Devices/CreativeAnimation/InterpolationTypes }
using { /Verse.org/Random }
using { /Fortnite.com/Characters }

# See https://dev.epicgames.com/documentation/en-us/uefn/create-your-own-device-in-verse for how to create a verse device.

# A Verse-authored creative device that can be placed in a level
hardbread_managerv2 := class(creative_device):


    @editable
    ForcedType : int = -1
    @editable
    IsInestable : logic = false
    var InestableHBIndex : int = -1
    var InestableState : int = 0
    var InestableNumHits : int = 0
    var InestableTotalHits : int = 0
    var InestabilityRunningTime : int = 9999
    var AssignedInestablePosition : int = -1
    @editable
    var Level : int = 0
    var WeakPointAux_Used : int = -1
    var HBType : int = 0
    var TotalTimeWithoutDamage : int = -1
    
    var Test_NumHits : int = 0

    var GameManager : game_manager = game_manager{}
    var AllPlayersNumHits_ForWeakPoint : [agent]int = map{}
    var AllPlayersNumHits : [agent]int = map{}
    var AllPlayersDamageAcumulated : [agent]float = map{}

    var DamageAcumulated : float = 0.0
    var ConsecutiveHits : int = 0
    var WeakPointExposed : logic = false

    var MaxHealth : float = 1000.0
    var TotalHealth : float = 1000.0
    var DamageToWeakPoint : float = 0.0
    var IsDamageAnimationPlaying : logic = false 
    var IsDeathAnimationPlaying : logic = false 

    @editable
    SlateBreadProp : creative_prop = creative_prop{}

    @editable
    MainManipulator : prop_manipulator_device = prop_manipulator_device{}
    

    #Death settings
    DecrementOnZ : float = 64.0
    var DelayToResurrect : float = 5.0
    var OriginalPosition : transform = transform{}
    #var AllHPieces : []creative_prop = array{}

    # Runs when the device is started in a running game
    OnBegin<override>()<suspends>:void=
        InitSubscribes()
        Sleep(1.0)
        ResetSetType(false)
        SetMaterial()
        Sleep(2.0)
        SavePosition()
        GenerateHealth()
        if(Level=HB_MAXLEVEL):
            set HBType=HB_PURE
        if(ForcedType<>-1):
            set HBType=ForcedType
            SetMaterial()

    #Inestability
    SetInestableHBIndex(Index : int) : void =
        set InestableHBIndex = Index 

    SetInestablePosition(NewPosition : int) : void =
        set AssignedInestablePosition = NewPosition
        set InestabilityRunningTime = 0

    InestabilityTick(MaxTime : int) : logic =
        var NeedToBeMove : logic = false
        set InestabilityRunningTime = InestabilityRunningTime + 1
        if(InestabilityRunningTime >= MaxTime):
            set NeedToBeMove = true
        return NeedToBeMove

    InitNewInestability(NewLevel : int) : void =
        set Level = NewLevel
        set InestabilityRunningTime = 0
        set InestableState = 0
        set InestableNumHits = 0
        set InestableTotalHits = 0
        SetMaterial()

    InestabilityHit(Agent : agent) : void =
        set InestableNumHits = InestableNumHits + 1
        set InestableTotalHits = InestableTotalHits + 1
        if(InestableTotalHits>50):
            set InestabilityRunningTime = InestabilityRunningTime + 5
        if(InestableState=0): #From Normal State
            RandomNum:=GetRandomInt(0,InestableNumHits)
            if(RandomNum>5):
                set InestableNumHits = 0
                RandomDecision:=GetRandomInt(0,10)
                if(RandomDecision>=4):
                    HandlerSpawnSuperState()
                else:
                    HandlerSpawnInestableDanger()
        else if(InestableState=1):
        else if(InestableState=2):
            if(FortCharacter:=Agent.GetFortCharacter[]):  
                GameManager.GeneralMobsManager.MobBehaviourHelper.DamageCharacter(FortCharacter,CalculeInestableCharacterDamage())
        else if(InestableState=3):
        SetMaterial()

    CalculeInestableCharacterDamage() : float =
        var DamageAmount : float = (Level-10)*15.0
        if(DamageAmount>=60.0):
            set DamageAmount = 59.0
        if(DamageAmount<0.0):
            set DamageAmount = 30.0
        return DamageAmount
        
    HandlerSpawnInestableDanger() : void =
        set InestableState = 3
        DangerTime:=GetRandomFloat(1.0,4.0)
        spawn:
            SpawnInestableState(DangerTime,2)

    HandlerSpawnSuperState() : void =
        spawn:
            GameManager.VFXManager.PlaySound(IHB_PHASE1_SOUND,0.0,false,option{SlateBreadProp.GetTransform()})
        SuperTime:=GetRandomFloat(1.5,4.5)
        set InestableState = 1
        SetMaterial()
        spawn:
            SpawnInestableState(SuperTime,1)

    HandlerSpawnNeutralState() : void =
        set InestableState = 0
        spawn:
            GameManager.VFXManager.PlaySound(IHB_PHASE0_SOUND,0.0,false,option{SlateBreadProp.GetTransform()})

    SpawnInestableState(DelayTime : float, Initial : int)<suspends> : void =
        if(InestableState = 3):
            spawn:
                GameManager.VFXManager.PlaySound(IHB_WARNING_SOUND,0.0,false,option{SlateBreadProp.GetTransform()})
            Sleep(1.5)
            spawn:
                GameManager.VFXManager.PlaySound(IHB_PHASE2_SOUND,0.0,false,option{SlateBreadProp.GetTransform()})
            Sleep(0.25)
        set InestableState = Initial
        SetMaterial()
        if(InestableState = 2):
            
        NewDecision:=GetRandomInt(0,10)
        Sleep(DelayTime)
        if(Initial=1):
            if(NewDecision<=7):
                HandlerSpawnNeutralState() 
            else if(NewDecision<=9):
                HandlerSpawnInestableDanger()
            else:
                HandlerSpawnSuperState()
        else if(Initial=2):
            if(NewDecision<=7):
                HandlerSpawnNeutralState() 
            else if(NewDecision<=9):
                HandlerSpawnSuperState()
            else:
                HandlerSpawnInestableDanger()
        SetMaterial()

    GetCentralWeakTransform(IsLastLevel : logic) : transform =
        if(not IsLastLevel?):
            XPosition:=SlateBreadProp.GetTransform().Translation.X
            YPosition:=SlateBreadProp.GetTransform().Translation.Y
            ZPosition:=SlateBreadProp.GetTransform().Translation.Z + 32.0
            return transform{Translation:=vector3{X:=XPosition,Y:=YPosition,Z:=ZPosition},Rotation:=SlateBreadProp.GetTransform().Rotation,Scale:=vector3{X:=1.0,Y:=1.0,Z:=1.0}}
        else:
            XPosition:=SlateBreadProp.GetTransform().Translation.X
            YPosition:=SlateBreadProp.GetTransform().Translation.Y
            ZPosition:=SlateBreadProp.GetTransform().Translation.Z + 64.0
            return transform{Translation:=vector3{X:=XPosition,Y:=YPosition,Z:=ZPosition},Rotation:=SlateBreadProp.GetTransform().Rotation,Scale:=vector3{X:=1.0,Y:=1.0,Z:=1.0}}


    SpawnHealthChecker()<suspends> : void =
        MaxTimeWithoutDamage := 30
        if(TotalTimeWithoutDamage=-1):
            set TotalTimeWithoutDamage=MaxTimeWithoutDamage
            loop:
                Sleep(1.0)
                set TotalTimeWithoutDamage=TotalTimeWithoutDamage-1
                if(TotalTimeWithoutDamage<=0):
                    ResetStats()
                    SetMaterial()
                    set TotalTimeWithoutDamage = -1
                    break
        else:
            set TotalTimeWithoutDamage=MaxTimeWithoutDamage

    ResetSetType(IsRurrect : logic) : void =
        var PosibleTypes : []int = array{}
        if(IsInestable?):
            set PosibleTypes = array{HB_TYPE4} #Inestable
        else:
            if(Level<5): 
                if(Level<2): 
                    set PosibleTypes = array{HB_TYPE0,HB_TYPE1} #Water. Flour.
                else:
                    set PosibleTypes = array{HB_TYPE0,HB_TYPE1,HB_TYPE2} #Water. Flour. Cheese
            else:
                set PosibleTypes = array{HB_TYPE0,HB_TYPE1,HB_TYPE2,HB_TYPE3} #Water. Flour. Cheese. Pimentos.
        if(IsRurrect?):
            #This code calcules the lowest num of type in the same level and chooses if as the new one
            var AllPosibilityArray : []int = array{} 
            for(PosibilityArray : PosibleTypes):
                set AllPosibilityArray = AllPosibilityArray + array{0}
            for(HB : GameManager.AllHardBreads, HB.Level = Level, HB.TotalHealth > 0.0, NumOfType := AllPosibilityArray[HB.HBType], set AllPosibilityArray[HB.HBType] = NumOfType + 1): #NumOfTypes
            var ChoosedType : int = HB_TYPE0 
            var LowestNum : int = 999
            for(TYPE := HB_TYPE0..AllPosibilityArray.Length-1, NumOfTypes := AllPosibilityArray[TYPE]):
                Log("Type: {TYPE} | NumOfTypes: {NumOfTypes}",LOG_LEVEL_DEBUG4)
                if(NumOfTypes < LowestNum):
                    set LowestNum = NumOfTypes
                    set ChoosedType = TYPE
            Log("Resurreted with type {ChoosedType} because there was only {LowestNum} in the same level",LOG_LEVEL_DEBUG4)
            set HBType = ChoosedType
        else: 
            RandomType := GetRandomInt(0,PosibleTypes.Length-1)
            if(Value := PosibleTypes[RandomType], set HBType = Value):
        if(Level=0, ForcedType=-1):
            set HBType=HB_BASETYPE
        if(Level=HB_MAXLEVEL):
            set HBType=HB_PURE
        if(ForcedType<>-1):
            set HBType=ForcedType

    SetMaterial() : void =
        if(IsInestable?):
            if(InestableState=0):
                SlateBreadProp.SetMaterial(Mat_HB_MutantMaterial, ?Index:=0)
            else if(InestableState=1):
                SlateBreadProp.SetMaterial(Mat_HB_MutantMaterialWeak, ?Index:=0)
            else if(InestableState=2):
                SlateBreadProp.SetMaterial(Mat_HB_MutantMaterialDangerous, ?Index:=0)
        else:
            if(Level<>0 or ForcedType<>-1):
                if(HBType=HB_TYPE0):
                    SlateBreadProp.SetMaterial(Mat_HB_Water, ?Index:=0)
                else if(HBType=HB_TYPE1):
                    SlateBreadProp.SetMaterial(Mat_HB_Flour, ?Index:=0)
                else if(HBType=HB_TYPE2):
                    SlateBreadProp.SetMaterial(Mat_HB_Cheese, ?Index:=0)
                else if(HBType=HB_TYPE3):
                    SlateBreadProp.SetMaterial(Mat_HB_Pimentos, ?Index:=0)
                else if(HBType=HB_TYPE4):
                    SlateBreadProp.SetMaterial(Mat_HB_MutantMaterial, ?Index:=0)
                else if(HBType=HB_PURE):
                    SlateBreadProp.SetMaterial(Mat_HB_Pure, ?Index:=0)
    
                
    SetWeakPointMaterial() : void =
        if(HBType=HB_TYPE0):
            SlateBreadProp.SetMaterial(Mat_HB_WaterWeak, ?Index:=0)
        else if(HBType=HB_TYPE1):
            SlateBreadProp.SetMaterial(Mat_HB_FlourWeak, ?Index:=0)
        else if(HBType=HB_TYPE2):
            SlateBreadProp.SetMaterial(Mat_HB_CheeseWeak, ?Index:=0)
        else if(HBType=HB_TYPE3):
            SlateBreadProp.SetMaterial(Mat_HB_PimentosWeak, ?Index:=0)
        else if(HBType=HB_TYPE4):
            SlateBreadProp.SetMaterial(Mat_HB_MutantMaterialWeak, ?Index:=0)
        else if(HBType=HB_PURE):
            SlateBreadProp.SetMaterial(Mat_HB_PureWeak, ?Index:=0)

    SetGameManager(InGameManager : game_manager) : void =
        set GameManager = InGameManager

    SavePosition() : void =
        set OriginalPosition = SlateBreadProp.GetTransform()

    GenerateHealth() : void =
        set MaxHealth = GameManager.PickaxeCalculator.MainFormula_GenerateHardBreadHealth(Level-1,HBType)
        set TotalHealth = MaxHealth
        set DamageToWeakPoint = TotalHealth*PercentageToWeakPoint

    ResetStats() : void =
        set AllPlayersNumHits_ForWeakPoint = map{}
        set AllPlayersDamageAcumulated = map{}
        set AllPlayersNumHits = map{}
        set DamageAcumulated = 0.0
        set ConsecutiveHits = 0
        set WeakPointExposed = false
        set TotalHealth = MaxHealth
        #NEW WEK POINT CODE
        TryFreeWeakPoint()

    TryFreeWeakPoint() : void =
        if(WeakPointAux_Used<>-1):
            spawn:
                GameManager.FreeWeakPoint(SlateBreadProp.GetTransform(),WeakPointAux_Used) 
        set WeakPointAux_Used = -1

    WeakPointEvent(Agent : agent) : void = 
        DamageDealt(Agent,true)

    #Damage num piece
    DamagedEvent(Agent : agent) : void =
        GameManager.VFXManager.Accolade_Hit.Award(Agent)
        MainManipulator.RestockResources()
        DamageDealt(Agent,false)

    RewardDestroyedHB() : void = 
        for(Index:=0..GetPlayspace().GetPlayers().Length-1, Player := GetPlayspace().GetPlayers()[Index], Player.IsActive[], Agent := agent[Player], AgentBakery := GameManager.GetAgentBakery_Transacted(Agent)?,NumHitsAccumulated := AllPlayersNumHits[Agent], PlayerDamageAcumulated := AllPlayersDamageAcumulated[Agent], DamageAcumulated > 0.0):
            var AllResourcersType : []int = array{}
            var AllResourcersAmmount : []float = array{}
            DamagePercentage := MaxHealth/PlayerDamageAcumulated
            AmountCoalMaterial:=CheckFinalCoalAward(Level,NumHitsAccumulated,DamagePercentage)
            if(AmountCoalMaterial>0.0):
                set AllResourcersType = AllResourcersType + array{COAL}
                set AllResourcersAmmount = AllResourcersAmmount + array{AmountCoalMaterial}
            AmountMutantMaterial:=CheckFinalMutantMaterialAward(Level,NumHitsAccumulated*1.0,DamagePercentage)
            if(AmountMutantMaterial>0.0):
                set AllResourcersType = AllResourcersType + array{MUTANTMATERIAL}
                set AllResourcersAmmount = AllResourcersAmmount + array{AmountMutantMaterial}
            AgentBakery.SaveResourcers(Agent,AllResourcersType,AllResourcersAmmount)
            AgentBakery.RefreshInventoryUI(AgentBakery.PlayerDataCache)
            Log("Player [{Index}] did {GetOptimizedMagnitudeLetter(DamagePercentage*100.0)}% damage to this HB",LOG_LEVEL_DEBUG4)
            spawn:
                AgentBakery.Delay_DisplayResourcersGather(AllResourcersAmmount,AllResourcersType,false,0.25)

    CalculeDamageDealtAndResourcers(Agent : agent, IsWeakPoint : logic) : float =
        var PickaxeDamage : float = 0.0
        #Log("ConsecutiveHits: {DamageToWeakPoint}/{DamageAcumulated}",LOG_LEVEL_UIs)
        #Log("WeakPoint is {WeakPoint}", LOG_LEVEL_UIs)
        #Check if it is weak point
        if(IsWeakPoint?):
            GameManager.VFXManager.Spawn_HardBreadCriticalStrike(Agent,HBType)
            SetMaterial()
            set WeakPointExposed = false
            TryFreeWeakPoint()
        else if(InestableState=1):
            GameManager.VFXManager.Spawn_HardBreadCriticalStrike(Agent,HBType)
        #Damage to the cube
        MaybeBakery:=GameManager.GetAgentBakery(Agent)
        if(ValidBakery:=MaybeBakery?):
            var ResultTuple : tuple(float,[]int,[]float,[]logic,logic) = (0.0,array{0},array{0.0},array{false},false)
            var NumHitsAccumulated_ForWeakPoint : int = 0
            if(Level<>0<#,Level<>HB_MAXLEVEL#>):
                #Num weak point Hits
                if(not IsInestable?):
                    if(Value:=AllPlayersNumHits_ForWeakPoint[Agent]):
                        set NumHitsAccumulated_ForWeakPoint = Value
                    set NumHitsAccumulated_ForWeakPoint = NumHitsAccumulated_ForWeakPoint + 1
                    if(set AllPlayersNumHits_ForWeakPoint[Agent] = NumHitsAccumulated_ForWeakPoint):
                        Log("NumHitsAccumulated_ForWeakPoint accumulated {NumHitsAccumulated_ForWeakPoint}", LOG_LEVEL_UIs)
                #Num total hits
                var NumHitsAccumulated : int =0
                if(Value:=AllPlayersNumHits[Agent]):
                    set NumHitsAccumulated = Value
                set NumHitsAccumulated = NumHitsAccumulated + 1
                #Only for last level HB
                if(Level=HB_MAXLEVEL, NumHitsAccumulated>15):
                    AmountMutantMaterial:=CheckFinalMutantMaterialAward(Level,NumHitsAccumulated*1.0,1.0)
                    set NumHitsAccumulated = 0
                    if(AmountMutantMaterial>0.0):
                        ValidBakery.SaveResourcers(Agent,array{MUTANTMATERIAL},array{AmountMutantMaterial})
                        ValidBakery.RefreshInventoryUI(ValidBakery.PlayerDataCache)
                        #Log("Player [{Index}] did {GetOptimizedMagnitudeLetter(DamagePercentage*100.0)}% damage to this HB",LOG_LEVEL_DEBUG4)
                        spawn:
                            ValidBakery.Delay_DisplayResourcersGather(array{AmountMutantMaterial},array{MUTANTMATERIAL},false,0.25)
                ############
                if(set AllPlayersNumHits[Agent] = NumHitsAccumulated):
                    Log("NumHitsAccumulated accumulated {NumHitsAccumulated}", LOG_LEVEL_UIs)
                #Get resourcers
                if(HBType = HB_PURE, IsWeakPoint?):
                    Log("WeakPoint Lv MAX Detected", LOG_LEVEL_DEBUG7)
                    for(Index:=0..GetPlayspace().GetPlayers().Length-1, Player := GetPlayspace().GetPlayers()[Index], OtherAgent := agent[Player], Player.IsActive[], OtherBakery:=GameManager.GetAgentBakery_Transacted(OtherAgent)? ,Character := Player.GetFortCharacter[], 512.0 > Distance(Character.GetTransform().Translation,SlateBreadProp.GetTransform().Translation)):
                        Log("Granting to player {Index}", LOG_LEVEL_DEBUG7)
                        set ResultTuple = OtherBakery.GetDamage_And_Resourcers(OtherAgent,NumHitsAccumulated,HBType,Level,IsWeakPoint,TotalHealth, ?IsInestable := IsInestable, ?InestableState := InestableState)
                        AllResourcersType := ResultTuple(1)
                        AllResourcersAmmount := ResultTuple(2)
                        AllLimitsReached := ResultTuple(3)
                        IsCritical := ResultTuple(4)
                        OtherBakery.DisplayDamageUI(OtherAgent, AllResourcersAmmount,AllResourcersType, AllLimitsReached,IsCritical)
                else: 
                    set ResultTuple = ValidBakery.GetDamage_And_Resourcers(Agent,NumHitsAccumulated,HBType,Level,IsWeakPoint,TotalHealth,  ?IsInestable := IsInestable, ?InestableState := InestableState)
                    AllResourcersType := ResultTuple(1)
                    AllResourcersAmmount := ResultTuple(2)
                    AllLimitsReached := ResultTuple(3)
                    IsCritical := ResultTuple(4)
                    ValidBakery.DisplayDamageUI(Agent, AllResourcersAmmount,AllResourcersType, AllLimitsReached,IsCritical)
            if(Level=0):
                Log("Level 0 HB Detected",LOG_LEVEL_UK)
                set ResultTuple = ValidBakery.GetDamage_And_Resourcers(Agent,0,HBType,Level,IsWeakPoint,TotalHealth,  ?IsInestable := IsInestable, ?InestableState := InestableState)
                AllResourcersType := ResultTuple(1)
                AllResourcersAmmount := ResultTuple(2)
                AllLimitsReached := ResultTuple(3)
                IsCritical := ResultTuple(4)
                #set ResultTuple = ValidBakery.GetDamage_And_Resourcers(Agent,0,HBType,Level,IsWeakPoint,TotalHealth)
                ValidBakery.DisplayDamageUI(Agent, AllResourcersAmmount,AllResourcersType, AllLimitsReached,IsCritical)
            if(IsInestable?, InestableState=1):
                AmountMutantMaterial:=CheckFinalMutantMaterialAward(Level,MinNum_AVGHit*1.0,1.0)*0.125
                if(AmountMutantMaterial>0.0):
                    ValidBakery.SaveResourcers(Agent,array{MUTANTMATERIAL},array{AmountMutantMaterial})
                    ValidBakery.RefreshInventoryUI(ValidBakery.PlayerDataCache)
                    #Log("Player [{Index}] did {GetOptimizedMagnitudeLetter(DamagePercentage*100.0)}% damage to this HB",LOG_LEVEL_DEBUG4)
                    spawn:
                        ValidBakery.Delay_DisplayResourcersGather(array{AmountMutantMaterial},array{MUTANTMATERIAL},false,0.25)
            set PickaxeDamage = ResultTuple(0)
            #Log("PickaxeDamage: {PickaxeDamage}",LOG_LEVEL_UIs)
            #Log("AllResourcersType.Length: {AllResourcersType.Length}",LOG_LEVEL_UIs)
            #Log("AllResourcersAmmount.Length: {AllResourcersAmmount.Length}",LOG_LEVEL_UIs)
            
            #Display resourcers
            
            #SET WEAK POINT
            if(not IsInestable?, DamageToWeakPoint <= DamageAcumulated or NumHitsAccumulated_ForWeakPoint > GameManager.MaxNumHitsToWeakPoint, not WeakPointExposed?):
                set DamageAcumulated = 0.0
                if(set AllPlayersNumHits_ForWeakPoint[Agent]=0):
                #RandomInt:=GetRandomInt(0,7)
                #Log("Weak point seleted: {WeakPoint}",LOG_LEVEL_UIs)
                GameManager.VFXManager.PlayWeakPointExposed(Agent)
                set WeakPointAux_Used = GameManager.FindWeakPointAuxiliar(Self)
                SetWeakPointMaterial()
                if(WeakPointAux_Used<>-1):
                    Log("Placing weak point", LOG_LEVEL_HB2)
                    if(Level=HB_MAXLEVEL):
                        Result:=GameManager.PlaceWeakPoint(Self,WeakPointAux_Used,?IsLastLevel:=true)
                        if(Result?):
                            set WeakPointExposed = true  
                    else:
                        Result:=GameManager.PlaceWeakPoint(Self,WeakPointAux_Used)
                        if(Result?):
                            set WeakPointExposed = true   
            else:
                if(Level<>0, not IsInestable?):
                    set DamageAcumulated = DamageAcumulated + PickaxeDamage
        return PickaxeDamage

    DamageDealt(Agent: agent, IsWeakPoint : logic) : void =
        spawn:
            SpawnHealthChecker()
        MaybeBakery:=GameManager.GetAgentBakery(Agent)
        if(ValidBakery:=MaybeBakery?):
            if(ValidBakery.IsOnUKTutorial?, UKTutorial:=ValidBakery.MaybeUKTutorial?):
                UKTutorial.PickHitEvent()
            if(TotalHealth>0.0, HealthUI := ValidBakery.MaybeEntityHealthUI?): #If it is death already, don´t do anything
                TotalDamage := CalculeDamageDealtAndResourcers(Agent, IsWeakPoint)   
                if(Level<>0, Level<>HB_MAXLEVEL, not IsInestable?):
                    set TotalHealth = TotalHealth - TotalDamage
                    #Save Agent amount Damage
                    if(LastDamageAmount:=AllPlayersDamageAcumulated[Agent], set AllPlayersDamageAcumulated[Agent] = LastDamageAmount + TotalDamage):
                    else if(set AllPlayersDamageAcumulated[Agent] = TotalDamage):
                #Log("TotalDamage: {TotalDamage}, TotalHealth: {TotalHealth}",LOG_LEVEL_UIs)
                if(TotalHealth<=0.0):
                    TryFreeWeakPoint()
                    set TotalHealth = 0.0   
                    HealthUI.SetInfo(0.0,MaxHealth,HBType,Level)  
                    RewardDestroyedHB()
                    #Play death effect
                    if(not IsDeathAnimationPlaying?):
                        set IsDeathAnimationPlaying = true
                        spawn:
                            AnimateDeathEffect(Agent,0.3)
                else:
                    HealthUI.SetInfo(TotalHealth,MaxHealth,HBType,Level)
                    #Play damage effect
                    if(not IsDamageAnimationPlaying?, TotalDamage > (TotalHealth*0.01)):
                        set IsDamageAnimationPlaying = true
                        spawn:
                            AnimateDamagedEffect()
        #Inestability
        if(IsInestable?):
            InestabilityHit(Agent)

    DestroyBread(EliminatorAgent : agent)<suspends> : void =
        AMobWillBeSpawned:=CheckRandomEncounter(EliminatorAgent)
        if(AMobWillBeSpawned?):
            Sleep(RandomMobMaxLifeTime+DelayToResurrect)
        else:
            Sleep(DelayToResurrect)
        InitResurect()

    CheckRandomEncounter(EliminatorAgent : agent) : logic =
        var TotalNumHits : int = 0
        var AMobWillBeSpawned : logic = false
        for(PlayersNumHits : AllPlayersNumHits):
            set TotalNumHits = TotalNumHits + PlayersNumHits
        if(GameManager.DebugOn? or TotalNumHits>=15):
            set AMobWillBeSpawned = GameManager.GeneralMobsManager.RollRandomPlantEncounter(EliminatorAgent, AdaptTransformFromVisualEmpty(OriginalPosition,-UpVisualEmptyHideZ), false, Level,TotalNumHits, Level)
        return AMobWillBeSpawned

    InitSubscribes<private>() : void =
        MainManipulator.HarvestingEvent.Subscribe(DamagedEvent)
        MainManipulator.RestockResources()
       
    InitResurect<private>()<suspends> : void =
        MoveTime : float = 2.0
        UpCrementOnZ := DecrementOnZ * 4
        ResetSetType(true)
        spawn:
            SlateBreadProp.MoveTo(OriginalPosition,MoveTime)
        Sleep(MoveTime/2.0)
        SetMaterial()
        CheckForOverlapedPlayers()
        Sleep(MoveTime/2.0)
        CheckForOverlapedPlayers()
        ResetStats()
        
    CheckForOverlapedPlayers() : void =
        for(Player : GetPlayspace().GetPlayers(), Character := Player.GetFortCharacter[],168.0>Distance(Character.GetTransform().Translation,SlateBreadProp.GetTransform().Translation)):
            NewTransform := vector3{X:=Character.GetTransform().Translation.X,Y:=Character.GetTransform().Translation.Y,Z:=Character.GetTransform().Translation.Z+128.0}
            if(Character.TeleportTo[NewTransform,Character.GetTransform().Rotation]):
            else:
                TeleportFailedAttempt(option{Character}, false, false, Character.GetTransform(), 64.0)

    AnimateDeathEffect<public>(Agent : agent, Delay : float)<suspends> : void =
        spawn:
            GameManager.VFXManager.Spawn_HardBreadDeath(Agent,SlateBreadProp,HBType,?ZVariation:=32.0)
        SetMaterial()
        EmptyTranslation := vector3{X:=0.0,Y:=0.0,Z:=0.0}
        X:=1.0+0.2
        
        #Roll is X: Red Axis OF THE OBJECT
        #Pitch is Y: Green Axis OF THE OBJECT
        #Yaw is Z: Blue axis OF THE OBJECT

        Rot1_RotationRoll : float = 20.0
        Rot2_RotationRoll : float = -40.0
        Rot3_RotationRoll : float = 40.0
        Rot4_RotationRoll : float = -20.0

        RotationApplied1 := MakeRotationFromYawPitchRollDegrees(0.0, 0.0,Rot1_RotationRoll)
        RotationApplied2 := MakeRotationFromYawPitchRollDegrees(0.0, 0.0,Rot2_RotationRoll)
        RotationApplied3 := MakeRotationFromYawPitchRollDegrees(0.0, 0.0,Rot3_RotationRoll)
        RotationApplied4 := MakeRotationFromYawPitchRollDegrees(0.0, 0.0,Rot4_RotationRoll)
        

        MovementKeyFrame1 : keyframe_delta = keyframe_delta #Key Frame 1 to base position

        {
            DeltaLocation := vector3{X:=0.0,Y:=0.0,Z:=-DecrementOnZ}
            DeltaRotation := RotationApplied1
            DeltaScale := vector3{X:=0.5,Y:=0.5,Z:=0.5}
            Time := 0.25
            Interpolation := EaseIn
        }

        MovementKeyFrame2 : keyframe_delta = keyframe_delta #Key to BouncePoint with Bounce Scale
        {
            DeltaLocation := vector3{X:=0.0,Y:=0.0,Z:=-DecrementOnZ}
            DeltaRotation := RotationApplied2
            DeltaScale := vector3{X:=0.5,Y:=0.5,Z:=0.5}
            Time := 0.25
            Interpolation := EaseIn
        }

        MovementKeyFrame3 : keyframe_delta = keyframe_delta #Key to BouncePoint with Bounce Scale
        {
            DeltaLocation := vector3{X:=0.0,Y:=0.0,Z:=-DecrementOnZ}
            DeltaRotation := RotationApplied3
            DeltaScale := vector3{X:=0.5,Y:=0.5,Z:=0.5}
            Time := 0.25
            Interpolation := EaseIn
        }

        MovementKeyFrame4 : keyframe_delta = keyframe_delta #Key to BouncePoint with Bounce Scale
        {
            DeltaLocation := vector3{X:=0.0,Y:=0.0,Z:=-DecrementOnZ}
            DeltaRotation := RotationApplied4
            DeltaScale := vector3{X:=0.5,Y:=0.5,Z:=0.5}
            Time := 0.25
            Interpolation := EaseIn
        }

        if(PropController := SlateBreadProp.GetAnimationController[]):
            ##Print("Playing death animation")
            # Make an array of key frames
            FortVaderKeyFrames :  []keyframe_delta = array { MovementKeyFrame1 , MovementKeyFrame2,  MovementKeyFrame3, MovementKeyFrame4}
            # Set the animation to play
            PropController.SetAnimation(FortVaderKeyFrames, ?Mode:=animation_mode.OneShot)
            # Play Animation
            #Sleep(Delay)
            if(IsDamageAnimationPlaying?):
                Sleep(Delay)
            if(not IsDamageAnimationPlaying?):
                PropController.Play()
                loop:
                    State := PropController.GetState()
                    if(State = animation_controller_state.Stopped):
                        set IsDeathAnimationPlaying = false 
                        spawn:
                            DestroyBread(Agent)
                        break
                    Sleep(0.04)

    AnimateDamagedEffect<public>()<suspends> : void =
        EmptyTranslation := vector3{X:=0.0,Y:=0.0,Z:=0.0}
        X:=1.0+0.2
        
        #Roll is X: Red Axis OF THE OBJECT
        #Pitch is Y: Green Axis OF THE OBJECT
        #Yaw is Z: Blue axis OF THE OBJECT

        Rot1_RotationRoll : float = 2.0
        Rot2_RotationRoll : float = -4.0
        Rot3_RotationRoll : float = 2.0

        NormalScale : vector3 = vector3{X:=1.0,Y:=1.0,Z:=1.0}

        RotationApplied1 := MakeRotationFromYawPitchRollDegrees(0.0, 0.0,Rot1_RotationRoll)
        RotationApplied2 := MakeRotationFromYawPitchRollDegrees(0.0, 0.0,Rot2_RotationRoll)
        RotationApplied3 := MakeRotationFromYawPitchRollDegrees(0.0, 0.0,Rot3_RotationRoll)
        

        MovementKeyFrame1 : keyframe_delta = keyframe_delta #Key Frame 1 to base position

        {
            DeltaLocation := EmptyTranslation
            DeltaRotation := RotationApplied1
            DeltaScale := NormalScale
            Time := 0.1
            Interpolation := EaseIn
        }

        MovementKeyFrame2 : keyframe_delta = keyframe_delta #Key to BouncePoint with Bounce Scale
        {
            DeltaLocation := EmptyTranslation 
            DeltaRotation := RotationApplied2
            DeltaScale := NormalScale
            Time := 0.1
            Interpolation := EaseIn
        }

        MovementKeyFrame3 : keyframe_delta = keyframe_delta #Key Frame 3 to go to the Destiny
        {
            DeltaLocation := EmptyTranslation 
            DeltaRotation := RotationApplied3
            DeltaScale := NormalScale
            Time := 0.1
            Interpolation := EaseIn
        }

        if(PropController := SlateBreadProp.GetAnimationController[]):
            ##Print("Playing damage animation")
            # Make an array of key frames
            FortVaderKeyFrames :  []keyframe_delta = array { MovementKeyFrame1 , MovementKeyFrame2,  MovementKeyFrame3}
            # Set the animation to play
            PropController.SetAnimation(FortVaderKeyFrames, ?Mode:=animation_mode.OneShot)
            # Play Animation
            PropController.Play()
            loop:
                State := PropController.GetState()
                if(State = animation_controller_state.Stopped):
                    set IsDamageAnimationPlaying = false 
                    break
                Sleep(0.04)


 
