# fill this file with your verse-script
using { /Fortnite.com/Devices }
using { /UnrealEngine.com/Temporary/SpatialMath }
using { /Verse.org/Simulation }
using { /Fortnite.com/Devices/CreativeAnimation }
using { /Fortnite.com/Devices/CreativeAnimation/InterpolationTypes }
using { /Verse.org/Random }

hexagame_pet := class():
    var MaybeHexaMiniGameManager : ?hexa_minigame_manager
    var PropAsset : creative_prop_asset = DefaultCreativePropAsset
    var MeleeAttackEffectAsset : creative_prop_asset = DefaultCreativePropAsset
    var RangeAttackEffectAsset : creative_prop_asset = DefaultCreativePropAsset
    var HealEffectAsset : creative_prop_asset = DefaultCreativePropAsset

    var AllPetsIndex : int = -1
    var MaybeProp : ?creative_prop = false
    
    var IsPlayerPet : logic
    var PetType : int

    var IsBeingUsed : logic = false

    var SuperJumpExecuted : logic = false

    var PackA_LV : int = -1
    var PackB_LV : int = -1

    var Purity : int = 1
    var Tier : int = 1

    var IsDeath : logic = false
    var Health : float = 0.0
    var MaxHealth : float = 100.0

    var HealingPerk : float = 0.0

    var LastAttack : float = -1.0
    #var TotalTime : float = 0.0
    
    var ActionSpeed : float = 1.0
    var Speed : float = 1.0

    var DefendRange : int = 0
    var AttackRange : int = 0

    var SupportNumActions : int = 0

    var AttackDamage : float = 10.0
    var Position : tuple(int,int) = (-1,-1)
    var LastPosition : tuple(int,int) = (-1,-1)
    var InitialPosition : tuple(int,int) = (-1,-1) 
    var AllMovementIntentions : []tuple(int,int) = array{}
    var IsMoving : logic = false
    #var IsMakingAction : logic = false

    var AllFriendliesFound : []hexagame_pet = array{}
    var AllEnemiesFound : []hexagame_pet = array{}

    var PivotVariation : vector3 = vector3{}
    var NormalScale : vector3 = vector3{}

    var ID : float = -1.0
    #var AngleFacingBack : float = 0.0

    <#
    CreateExplosion()<suspends> : void =
        if(HexaMiniGameManager := MaybeHexaMiniGameManager?, BakeryMachinesManager:=HexaMiniGameManager.MaybeBakeryMachinesManager?):
            if:
                HexProp := HexaMiniGameManager.AllHexPositions[Position(0)][Position(1)]
            then:
                HexTransform:=HexProp.GetTransform()

                TupleSpawn:=SpawnProp(BakeryMachinesManager.GameManager.VFXManager.ExplosionEffect,HexTransform)
                MaybeCreativeProp:=TupleSpawn(0)
                if(ValidProp := MaybeCreativeProp?):
                    Sleep(3.0)
                    ValidProp.Dispose()#>


    <#AvoidClash() : void = #Clash happens when two pets are in the same position
        if(LastPosition(0)<>-9,LastPosition(1)<>-9, HexaMiniGameManager := MaybeHexaMiniGameManager?):
            #Print("Executing anti-clash")
            spawn:
                CreateExplosion()
            MovementApproved:=HexaMiniGameManager.CheckIfPositionIsEmpty(LastPosition)
                if(MovementApproved?):
                    ForcePosition(LastPosition(0),LastPosition(1))
                else:
                    TupleDown:=HexaMiniGameManager.GetNearDownPositions(Position(0),Position(1))
                    TupleSides:=HexaMiniGameManager.GetNearSidePositions(Position(0),Position(1))
                    TupleUp:=HexaMiniGameManager.GetNearSidePositions(Position(0),Position(1))

                    Movement0:=TupleDown(0)
                    Movement1:=TupleDown(1)
                    Movement2:=TupleSides(0)
                    Movement3:=TupleSides(1)
                    Movement4:=TupleUp(0)
                    Movement5:=TupleUp(1)

                    TupleMovementApproved_0:=HexaMiniGameManager.CheckIfPositionIsEmpty(Movement0)
                    TupleMovementApproved_1:=HexaMiniGameManager.CheckIfPositionIsEmpty(Movement1)
                    TupleMovementApproved_2:=HexaMiniGameManager.CheckIfPositionIsEmpty(Movement2)
                    TupleMovementApproved_3:=HexaMiniGameManager.CheckIfPositionIsEmpty(Movement3)
                    TupleMovementApproved_4:=HexaMiniGameManager.CheckIfPositionIsEmpty(Movement4)
                    TupleMovementApproved_5:=HexaMiniGameManager.CheckIfPositionIsEmpty(Movement5)

                    if(TupleMovementApproved_0?):
                        ForcePosition(Movement0(0),Movement0(1))
                    else if(TupleMovementApproved_1?):
                        ForcePosition(Movement1(0),Movement1(1))
                    else if(TupleMovementApproved_2?):
                        ForcePosition(Movement2(0),Movement2(1))
                    else if(TupleMovementApproved_3?):
                        ForcePosition(Movement3(0),Movement3(1))
                    else if(TupleMovementApproved_4?):
                        ForcePosition(Movement4(0),Movement4(1))
                    else if(TupleMovementApproved_5?):
                        ForcePosition(Movement5(0),Movement5(1))
                    else:
                        PrintFatalError("There was no posible rollback, trying a stupid teleport")
                        var IndexRow : int = 0
                        var IndexColumn : int = 0
                        var ValidPositionToMove : tuple(int,int) = (-1,-1)
                        if(IsPlayerPet?):
                            for(Row:=0..HexaMiniGameManager.HexaNumRows-1):
                                for(Column:=0..HexaMiniGameManager.HexaNumColumns-1):
                                    if(MaybePet:=HexaMiniGameManager.AllHexBoxes[Row][Column]):
                                        if(ValidPet:=MaybePet?):
                                        else if(ValidPositionToMove(0) = -1):
                                            set ValidPositionToMove = (Row,Column)
                        else:
                            var NumRow : int = HexaMiniGameManager.HexaNumRows-1
                            loop:
                                for(Column:=0..HexaMiniGameManager.HexaNumColumns-1):
                                    if(MaybePet:=HexaMiniGameManager.AllHexBoxes[NumRow][Column]):
                                        if(ValidPet:=MaybePet?):
                                        else if(ValidPositionToMove(0) = -1):
                                            set ValidPositionToMove = (NumRow,Column)
                                            break
                                set NumRow = NumRow - 1
                                if(NumRow < 0):
                                    break
                        #Destroy if it is imposible
                        if(ValidPositionToMove(0)<>-1,ValidPositionToMove(1)<>-1):
                            Die()
                        else:
                            ForcePosition(ValidPositionToMove(0),ValidPositionToMove(1))
                                    
                        #>




    #Only for clash use
    <#
    ForcePosition(ForcedPosition : tuple(int,int)) : void = 
        set IsMoving = true
        if(ForcedPosition(0)<>-1, ForcedPosition(1)<>-1, ValidProp := MaybeProp?):
            if:
                HexaMiniGameManager := MaybeHexaMiniGameManager?
                set HexaMiniGameManager.AllHexBoxes[ForcedPosition(0)][ForcedPosition(1)]=option{Self}
                set HexaMiniGameManager.AllHexBoxes[Position(0)][Position(1)]=false
                OldHexProp := HexaMiniGameManager.AllHexPositions[Position(0)][Position(1)]
                HexProp := HexaMiniGameManager.AllHexPositions[ForcedPosition(0)][ForcedPosition(1)]
            then:
                HexTransform:=HexProp.GetTransform()
                Translation : vector3 = HexTransform.Translation+PivotVariation
                NewTransform:=transform{Translation:=Translation,Rotation:=HexProp.GetTransform().Rotation,Scale:=NormalScale}

                CurrentTranslation := ValidProp.GetTransform().Translation

                ## New code
                set LastPosition = Position
                set Position = ForcedPosition
                set AllMovementIntentions = array{} #Why is this here? it make no sense to put the last position BEFORE moving the prop!
                ##
                if(ValidProp.TeleportTo[HexTransform]):
                
            <#
            set LastPosition = Position
            set Position = NewPosition
            set AllMovementIntentions = array{}#> #Why is this here? it make no sense to put the last position BEFORE moving the prop!
        set IsMoving = false#>

    InitHexaPet() : void =
        DespawnPetProp()
        set IsDeath = false
        set IsMoving = false
        set Health = MaxHealth
        set SuperJumpExecuted = false
        if(Tier=1):
            set NormalScale = PetScale_Tier1
            set PivotVariation = PetPivotVariation_Tier1
        if(Tier=2):
            set NormalScale = PetScale_Tier2
            set PivotVariation = PetPivotVariation_Tier2
        if(Tier=3):
            set NormalScale = PetScale_Tier3
            set PivotVariation = PetPivotVariation_Tier3

        if(Position(0) = -1 or Position(1) = -1):
           #Print("FATAL ERROR ON HEXA PET INIT. Must set a valid position!")
        else:
            if:
                HexaMiniGameManager := MaybeHexaMiniGameManager?
                set HexaMiniGameManager.AllHexBoxes[Position(0)][Position(1)]=option{Self}
            then:
                set InitialPosition = Position
                SpawnPetProp()
        #Set health
        if(HexaMiniGameManager := MaybeHexaMiniGameManager?, ValidProp := MaybeProp?):
            SetHealthUI(ValidProp,MaxHealth,Health, IsPlayerPet)
            SetStar(ValidProp,Tier)

    DespawnPetProp() : void =
        if:
            HexaMiniGameManager := MaybeHexaMiniGameManager?
            HexProp := HexaMiniGameManager.AllHexPositions[Position(0)][Position(1)]
        then:
            HexProp.SetMaterial(Materials.MI_HexaBaseNoFresnel, ?Index:=2)
        if(ValidProp := MaybeProp?, ValidProp.IsValid[]):
            ValidProp.Dispose()

    SpawnPetProp() : void =
        if(Tier=1):
            set NormalScale = PetScale_Tier1
            set PivotVariation = PetPivotVariation_Tier1
        if(Tier=2):
            set NormalScale = PetScale_Tier2
            set PivotVariation = PetPivotVariation_Tier2
        if(Tier=3):
            set NormalScale = PetScale_Tier3
            set PivotVariation = PetPivotVariation_Tier3
        if:
            HexaMiniGameManager := MaybeHexaMiniGameManager?
            HexProp := HexaMiniGameManager.AllHexPositions[Position(0)][Position(1)]
        then:
            Translation:=HexProp.GetTransform().Translation
            var Rotation:rotation=rotation{}
            if(not IsPlayerPet?):
                set Rotation = MakeRotationFromYawPitchRollDegrees(180.0, 0.0, 0.0)
            InitTranslation :=Translation+PivotVariation
            InitTransform:=transform{Translation:=InitTranslation,Rotation:=Rotation,Scale:=NormalScale}
            if(ValidProp := MaybeProp?):
                ValidProp.Dispose()
                Tuple:=SpawnProp(PropAsset,InitTransform)
                if(ValidProp1 := Tuple(0)?, set MaybeProp = option{ValidProp1}):
                    SetHealthUI(ValidProp1,MaxHealth,Health,IsPlayerPet)
                    SetStar(ValidProp1,Tier)
            else:
                Tuple:=SpawnProp(PropAsset,InitTransform)
                if(ValidProp2 := Tuple(0)?, set MaybeProp = option{ValidProp2}):
                    SetHealthUI(ValidProp2,MaxHealth,Health,IsPlayerPet)
                    SetStar(ValidProp2,Tier)
            if(not IsPlayerPet?):
                HexProp.SetMaterial(Materials.MI_HexaBaseRed, ?Index:=2)
            else:
                HexProp.SetMaterial(Materials.MI_HexaBaseGreen, ?Index:=2)
            
        
    SetIsBeingUsed(InIsBeingUsed : logic) : void =
        set IsBeingUsed = InIsBeingUsed

    ResetInitialPosition() : void =
        set InitialPosition=(-1,-1)

                
    SetPosition(InPosition : tuple(int,int)) : void =
        set LastPosition = Position
        set Position = InPosition
        if:
            HexaMiniGameManager := MaybeHexaMiniGameManager?
            HexProp := HexaMiniGameManager.AllHexPositions[Position(0)][Position(1)]
            ValidProp := MaybeProp?
        then:
            if(set HexaMiniGameManager.AllHexBoxes[Position(0)][Position(1)] = option{Self}):
            HexTransform:=HexProp.GetTransform()
            Translation := HexTransform.Translation+PivotVariation
            Transform:=transform{Translation:=Translation,Rotation:=HexProp.GetTransform().Rotation,Scale:=NormalScale}
            if(ValidProp.TeleportTo[Transform]):
                
    End() : void =
        set Health = 0.0
        set Position = (-9,-9)
        set LastPosition = (-9,-9)
        set IsBeingUsed = false
        set SuperJumpExecuted = false
        set AllMovementIntentions = array{}
        set IsMoving = false
        if(ValidProp := MaybeProp?, ValidProp.IsValid[]):
            ValidProp.Dispose()
        
    <#
    RunPet()<suspends> : void =
        set Health = MaxHealth
        set SupportNumActions = 0
        if(HexaMiniGameManager := MaybeHexaMiniGameManager?, HexProp := HexaMiniGameManager.AllHexPositions[Position(0)][Position(1)]):
            if(IsMoving?):
                PrintFatalError("Runing a pet, this pet is already moving before game: PetType {PetType} | Position [{Position(0)}][{Position(1)}]")
            if(Health<=0.0):
                PrintFatalError("Runing a pet, this pet is dead: PetType {PetType} | Position [{Position(0)}][{Position(1)}]")
            if(IsMakingAction?):
                PrintFatalError("Runing a pet, this pet is already making an action before game: PetType {PetType} | Position [{Position(0)}][{Position(1)}]")
            loop:
                UseTurn()
                if(Health<=0.0 or not HexaMiniGameManager.GameIsRunning?):
                    break
                Sleep(0.04)
                set TotalTime = TotalTime + 0.04
        else:
            PrintFatalError("FatalError Runing a pet, HexaMiniGameManager invalid or HexProp[{Position(0)}][{Position(1)}] invalid")#>
    InitPet() : void=
        set Health = MaxHealth
        set SupportNumActions = 0
        set IsMoving = false
        set IsDeath = false

    FindFriendliesAndFoes() : void =
        set AllFriendliesFound = array{} #EMPTY IT BEFORE
        set AllEnemiesFound = array{} #EMPTY IT BEFORE TO NOT DO A INFINITE LOOP
        if(HexaMiniGameManager := MaybeHexaMiniGameManager?):
            AllHexaPets:=HexaMiniGameManager.AllHexaPlayerPets+HexaMiniGameManager.AllHexaAIPets
            #Friendlies
            for(OtherPet : AllHexaPets): #Check all other pets
                #Position
                OtherPetPosition := OtherPet.GetPosition()
                PetPosition := GetPosition()
                #Faction
                OtherPetFaction := OtherPet.GetFaction()
                PetFaction := GetFaction()

                #Find enemies and frienly
                if:
                    OtherPetPosition(0)<>PetPosition(0) or OtherPetPosition(1)<>PetPosition(1) #If positions are different, they are a different pet
                    OtherPet.Health > 0.0 #He is not dead
                    OtherPet.IsBeingUsed?
                then:
                    if(OtherPet.IsPlayerPet=IsPlayerPet):
                        set AllFriendliesFound = AllFriendliesFound + array{OtherPet}
                    else:
                        set AllEnemiesFound = AllEnemiesFound + array{OtherPet}

    UseTurn() : void =
        #Find friendlies and foes        
        #if(IsMoving?, PetType=DRUID):
        #    #Print("Pet is moving!")
        if(HexaMiniGameManager := MaybeHexaMiniGameManager?):
            set AllMovementIntentions = array{}
            if(Health>0.0,  not IsMoving?):
                if(SuperJumpExecuted? or PetType <> HUNTER):
                    if(PetType<>DRUID): #Move if it is posible
                        SearchObjetive(true)
                    else: #Move if it is posible
                        if(SupportNumActions >= 3): #Now, lest buff the healing perk every 3 attacks
                            set SupportNumActions = 0
                            spawn:
                                HealAllAllies()
                        SearchObjetive(true)
                        <#
                        ObjetiveFound:=SearchObjetive(true) #Lets no find never allies with the druid
                        if(not ObjetiveFound?):
                            SearchObjetive(true)#>
                else if(not SuperJumpExecuted?):
                    Success:=TrySuperJump()
                    if(not Success?):
                        SearchObjetive(true)
        

    TrySuperJump() : logic =
        var Success : logic = false
        if(HexaMiniGameManager := MaybeHexaMiniGameManager?):
            var RowObjetive : int = HexaMiniGameManager.HexaNumRows-1
            if(IsPlayerPet?):
                set RowObjetive = HexaMiniGameManager.HexaNumRows-1
            else:
                set RowObjetive = 0
            if(MaybePosition:=HexaMiniGameManager.AllHexBoxes[RowObjetive][Position(1)]):
                if(ValidPet := MaybePosition?):
                    #Not empty ry side positions
                    TuplePositions:=HexaMiniGameManager.GetNearSidePositions(RowObjetive,Position(1))
                    PositionChoosed:=TuplePositions(0)
                    if(MaybePosition2:=HexaMiniGameManager.AllHexBoxes[PositionChoosed(0)][PositionChoosed(1)]):
                        if(ValidPet2 := MaybePosition2?):
                            TheOtherSide:=TuplePositions(1)
                            if(MaybePosition3:=HexaMiniGameManager.AllHexBoxes[TheOtherSide(0)][TheOtherSide(1)]):
                                if(ValidPet3 := MaybePosition3?):
                                #It is empty!
                                else if(PositionProp:=HexaMiniGameManager.AllHexPositions[TheOtherSide(0)][TheOtherSide(1)]):
                                    set SuperJumpExecuted = true
                                    set Success = true
                                    SetNewPosition(TheOtherSide(0),TheOtherSide(1))
                        #It is empty!
                        else if(PositionProp:=HexaMiniGameManager.AllHexPositions[PositionChoosed(0)][PositionChoosed(1)]):
                            set SuperJumpExecuted = true
                            set Success = true
                            SetNewPosition(PositionChoosed(0),PositionChoosed(1))
                #It is empty!
                else if(PositionProp:=HexaMiniGameManager.AllHexPositions[RowObjetive][Position(1)]):
                    set SuperJumpExecuted = true
                    set Success = true
                    SetNewPosition(RowObjetive,Position(1))
        return Success
                

    MoveProp(NewTranslation : vector3, CurrentTranslation : vector3)<suspends> : void =
        ObjetiveRotation:=GetForwardRotation_Vector2(NewTranslation,CurrentTranslation)
                NewTransform : transform = transform{
                    Translation:=NewTranslation}
                AnimateMovement(false,NewTransform,ObjetiveRotation)


    Move() : void =
        ##Print("Trying to move")
        var Index : int = 0
        loop:
            if(MovementIntention:=AllMovementIntentions[Index], HexaMiniGameManager := MaybeHexaMiniGameManager?):
                MovementApproved:=HexaMiniGameManager.CheckIfPositionIsEmpty(MovementIntention)
                if(MovementApproved?):
                    SetNewPosition(MovementIntention)
                    break
                else:
                    set Index = Index +1 
                    if(Index > AllMovementIntentions.Length-1):
                        if(AllMovementIntentions.Length=0):
                            #Print("Valid position to move not because is EMPTY!")
                        else:
                            #Print("Valid position to move not found")
                            ##Print("From [{Position(0)}][{Position(1)}] can not move to AllMovementIntentions:")
                            for(Option : AllMovementIntentions):
                                #Print("{Option(0)},{Option(1)} not valid because is full")
                        break
            else:
                PrintWarning("Movement intention {Index} not found")
                break

    AttackOfDefend(Attack : logic, ObjetivePet : hexagame_pet) : void =
        if(not IsMoving?):
            if(Attack?):
                #if(PetType=TANK):
                #    #Print("Attacking")
                if(PetType=ADC or PetType=DRUID):
                    set IsMoving = true
                    spawn:
                        Execute_RangeAttack(ObjetivePet)
                    if(PetType=DRUID):
                        #set SupportNumActions = SupportNumActions + 1
                else:
                    set IsMoving = true
                    spawn:
                        Execute_MeleeAttack(ObjetivePet)
            else: #DEFEND
                if:
                    ValidProp := MaybeProp?
                    ObjetiveProp := ObjetivePet.MaybeProp?
                    DistanceBetweenPets:=Floor[DistanceXY(ObjetiveProp.GetTransform().Translation,ValidProp.GetTransform().Translation)]
                    DefendRange > DistanceBetweenPets
                then:
                    if(ObjetivePet.MaxHealth=ObjetivePet.Health):
                        SearchObjetive(true)
                    else:
                        set IsMoving = true
                        spawn:
                            Execute_RangeHeal(ObjetivePet)
                        #Print("SupportNumActions: {SupportNumActions}")

    Execute_MeleeAttack(ObjetivePet : hexagame_pet)<suspends> : void =
        if(<#LastAttack + ActionSpeed < TotalTime, #>ValidProp := MaybeProp?, ValidObjetivePet := ObjetivePet.MaybeProp?):
            ObjetiveRotation:=GetForwardRotation_Vector2(ValidObjetivePet.GetTransform().Translation,ValidProp.GetTransform().Translation)
            
            NewTransform : transform = transform{
                Translation:=ValidProp.GetTransform().Translation,
                Rotation:=ObjetiveRotation,
                Scale:=ValidProp.GetTransform().Scale}

            #ValidProp.MoveTo(NewTransform,0.1)

            if:
                DistanceBetweenPets:=Floor[DistanceXY(ValidObjetivePet.GetTransform().Translation,ValidProp.GetTransform().Translation)]
                AttackRange > DistanceBetweenPets
            then:
                #set LastAttack = TotalTime
                AnimateMeleeAttack(ObjetivePet)
        set IsMoving = false
        set SupportNumActions = SupportNumActions + 1

    AnimateMeleeAttack(ObjetivePet : hexagame_pet)<suspends> : void =
        if(ValidProp := MaybeProp?, ValidProp.IsValid[], ValidObjetivePet := ObjetivePet.MaybeProp?):
            ViewRotation:=GetForwardRotation_Vector2(ValidObjetivePet.GetTransform().Translation,ValidProp.GetTransform().Translation)
            PointingVector:=ViewRotation.GetLocalForward()*32.0
            PrepareAttackVector:=-PointingVector/2.0
            var TotalAnimationTime : float = 0.0

            FacingEnemyTime : float = (ActionSpeed/8.0)*0.5
            FacingDelayMargin : float = (ActionSpeed/8.0)*0.25
            ActionFrame0 : float = (ActionSpeed/8.0)*4
            ActionFrame1 : float = (ActionSpeed/8.0)*1.5
            ActionFrame2 : float = (ActionSpeed/8.0)*1.5
            PostAttackTime : float = (ActionSpeed/8.0)*0.25

            FacingEnemyTransform : transform = transform{
                Translation:=ValidProp.GetTransform().Translation,
                Rotation:=ViewRotation,
                Scale:=ValidProp.GetTransform().Scale}

            ValidProp.MoveTo(FacingEnemyTransform,FacingEnemyTime)
            Sleep(FacingDelayMargin)

            
            MovementKeyFrame1 : keyframe_delta = keyframe_delta #Key Frame 1 to begining
                {
                    DeltaLocation := PrepareAttackVector
                    DeltaScale := NeutralScale
                    DeltaRotation := rotation{}
                    Time := ActionFrame0
                    Interpolation := EaseOut
                }


                MovementKeyFrame2 : keyframe_delta = keyframe_delta #Key Frame 1 to begining
                {
                    DeltaLocation := PointingVector+(PointingVector/2.0)
                    DeltaScale := NeutralScale
                    DeltaRotation := rotation{}
                    Time := ActionFrame1
                    Interpolation := EaseIn
                    
                }

                MovementKeyFrame3 : keyframe_delta = keyframe_delta #Key Frame 1 to begining
                {
                    DeltaLocation := -PointingVector
                    DeltaScale := NeutralScale
                    DeltaRotation := rotation{}
                    Time := ActionFrame2
                    Interpolation := EaseIn
                }

                if(PropController := ValidProp.GetAnimationController[]):
                    # Make an array of key frames
                    FortVaderKeyFrames :  []keyframe_delta = array { MovementKeyFrame1 , MovementKeyFrame2,  MovementKeyFrame3}
                    # Set the animation to play
                    PropController.SetAnimation(FortVaderKeyFrames, ?Mode:=animation_mode.OneShot)
                    # Play Animation
                    PropController.Play()
                    #f(ValidEffectSound := MaybeEffectSound?):
                    #    ValidEffectSound.Play()
                    loop:
                        State := PropController.GetState()
                        if(State = animation_controller_state.Stopped):
                            break
                        Sleep(0.04)
                        set TotalAnimationTime = TotalAnimationTime + 0.04
                        if(TotalAnimationTime>4.0):
                            break
                if(Health>0.0):
                    if(HexaMiniGameManager := MaybeHexaMiniGameManager?, BakeryMachinesManager := HexaMiniGameManager.MaybeBakeryMachinesManager?, Agent:=HexaMiniGameManager.MaybeAgent?):
                        spawn:
                            BakeryMachinesManager.GameManager.VFXManager.PlaySound(DAMAGE_SOUND,0.0,option{Agent},false)
        
                    ObjetivePet.DealDamage(AttackDamage)
                else:
                    PlayDieEffect()
                #set LastAttack = TotalTime
            Sleep(PostAttackTime)
            set IsMoving = false
            

    Execute_RangeHeal(ObjetivePet : hexagame_pet)<suspends> : void =
        if(<#LastAttack + ActionSpeed < TotalTime,#>ValidProp := MaybeProp?, ValidProp.IsValid[], ValidObjetivePet := ObjetivePet.MaybeProp?):
            ObjetiveRotation:=GetForwardRotation_Vector2(ValidObjetivePet.GetTransform().Translation,ValidProp.GetTransform().Translation)
            
            NewTransform : transform = transform{
                Translation:=ValidProp.GetTransform().Translation,
                Rotation:=ObjetiveRotation,
                Scale:=ValidProp.GetTransform().Scale}

            PreAttack := (ActionSpeed*2.0/10.0)
            PostAttack := (ActionSpeed*8.0/10.0)

            ValidProp.MoveTo(NewTransform,PreAttack)

            if:
                DistanceBetweenPets:=Floor[DistanceXY(ValidObjetivePet.GetTransform().Translation,ValidProp.GetTransform().Translation)]
                DefendRange > DistanceBetweenPets
            then:
                #set LastAttack = TotalTime
                ##Print("ATTACKING")
                if(HexaMiniGameManager := MaybeHexaMiniGameManager?, BakeryMachinesManager := HexaMiniGameManager.MaybeBakeryMachinesManager?, Agent:=HexaMiniGameManager.MaybeAgent?):
                    BakeryMachinesManager.GameManager.VFXManager.HexaPetRangedHeal_AudioPlayer.Play(Agent)
                spawn:
                    GenerateHealProjectile(ObjetivePet)
            Sleep(PostAttack)
        set IsMoving = false
        set SupportNumActions = SupportNumActions + 1


    Execute_RangeAttack(ObjetivePet : hexagame_pet)<suspends> : void =
        ##Print("Executing Execute_RangeAttack")
        if(<#LastAttack + ActionSpeed < TotalTime,#>ValidProp := MaybeProp?, ValidProp.IsValid[], ValidObjetivePet := ObjetivePet.MaybeProp?):
            #Face objetive
            ObjetiveRotation:=GetForwardRotation_Vector2(ValidObjetivePet.GetTransform().Translation,ValidProp.GetTransform().Translation)
            
            NewTransform : transform = transform{
                Translation:=ValidProp.GetTransform().Translation,
                Rotation:=ObjetiveRotation,
                Scale:=ValidProp.GetTransform().Scale}

            PreAttack := (ActionSpeed*1.0/5.0)
            PostAttack := (ActionSpeed*4.0/5.0)

            ValidProp.MoveTo(NewTransform,PreAttack)

            if:
                DistanceBetweenPets:=Floor[DistanceXY(ValidObjetivePet.GetTransform().Translation,ValidProp.GetTransform().Translation)]
                AttackRange > DistanceBetweenPets
            then:
                ##Print("ATTACKING")
                #set LastAttack = TotalTime
                if(HexaMiniGameManager := MaybeHexaMiniGameManager?, BakeryMachinesManager := HexaMiniGameManager.MaybeBakeryMachinesManager?, Agent:=HexaMiniGameManager.MaybeAgent?):
                    BakeryMachinesManager.GameManager.VFXManager.HexaPetRanged_AudioPlayer.Play(Agent)
                spawn:
                    GenerateProjectile(ObjetivePet)
            Sleep(PostAttack)
        set IsMoving = false
        set SupportNumActions = SupportNumActions + 1
 

    HealAllAllies()<suspends> : void =
        FindFriendliesAndFoes()
        set SupportNumActions = 0
        for(Objetive : AllFriendliesFound, Objetive.Health>0.0):
            spawn:
                GenerateHealProjectile(Objetive)
            Sleep(0.1)

    GenerateHealProjectile(ObjetivePet : hexagame_pet)<suspends> : void =
        ProjectileLimitTime:float=5.0
        var TotalProjectileTime:float = 0.0

        if(ValidProp := MaybeProp?, ValidObjetivePet := ObjetivePet.MaybeProp?):

            Tuple:=SpawnProp(HealEffectAsset,ValidProp.GetTransform())
            DistanceBetween:=Distance(ValidProp.GetTransform().Translation, ValidObjetivePet.GetTransform().Translation)
            TravelTime:=DistanceBetween/HexaGameHealProjectileSpeed
            
            FirstStageTime:=(TravelTime/10.0)*1.0
            MidStageTime:=(TravelTime/10.0)*8.0
            FinalStageTime:=(TravelTime/10.0)*1.0
            ##Print("FirstStageTime: {FirstStageTime} | MidStageTime: {MidStageTime} | FinalStageTime: {FinalStageTime}")

            FirstStageTranslation := vector3{X:=ValidProp.GetTransform().Translation.X,Y:=ValidProp.GetTransform().Translation.Y,Z:=ValidProp.GetTransform().Translation.Z+HexaGameProjectileElevation}
            VectorToPoint1 := FirstStageTranslation-ValidProp.GetTransform().Translation

            MidStageTranslation := vector3{X:=ValidObjetivePet.GetTransform().Translation.X,Y:=ValidObjetivePet.GetTransform().Translation.Y,Z:=ValidObjetivePet.GetTransform().Translation.Z+HexaGameProjectileElevation}
            VectorToPoint2 := MidStageTranslation-(ValidProp.GetTransform().Translation+VectorToPoint1)

            FinalStageTranslation := vector3{X:=ValidObjetivePet.GetTransform().Translation.X,Y:=ValidObjetivePet.GetTransform().Translation.Y,Z:=ValidObjetivePet.GetTransform().Translation.Z}
            VectorToPoint3 := vector3{Z:=-2.0*HexaGameProjectileElevation}

            if(ProjectileProp := Tuple(0)?):
                MovementKeyFrame1 : keyframe_delta = keyframe_delta #Key Frame 1 to begining
                {
                    DeltaLocation := VectorToPoint1
                    DeltaScale := NeutralScale
                    DeltaRotation := rotation{}
                    Time := FirstStageTime
                    Interpolation := EaseOut
                }

                MovementKeyFrame2 : keyframe_delta = keyframe_delta #Key Frame 1 to begining
                {
                    DeltaLocation := VectorToPoint2
                    DeltaScale := NeutralScale
                    DeltaRotation := rotation{}
                    Time := MidStageTime
                    Interpolation := EaseIn
                    
                }

                MovementKeyFrame3 : keyframe_delta = keyframe_delta #Key Frame 1 to begining
                {
                    DeltaLocation := VectorToPoint3
                    DeltaScale := NeutralScale
                    DeltaRotation := rotation{}
                    Time := FinalStageTime
                    Interpolation := EaseIn
                }


                if(ProjectilePropController := ProjectileProp.GetAnimationController[]):
                    # Make an array of key frames
                    FortVaderKeyFrames :  []keyframe_delta = array { MovementKeyFrame1 , MovementKeyFrame2,  MovementKeyFrame3}
                    # Set the animation to play
                    ProjectilePropController.SetAnimation(FortVaderKeyFrames, ?Mode:=animation_mode.OneShot)
                    # Play Animation
                    var MidPointReached : logic = false
                    ##Print("Projectile launched and moved")
                    ProjectilePropController.Play()
                    #f(ValidEffectSound := MaybeEffectSound?):
                    #    ValidEffectSound.Play()
                    loop:
                        State := ProjectilePropController.GetState()
                        if(State = animation_controller_state.Stopped):
                            ##Print("Projectile reached destiny")
                            break
                        Sleep(0.04)
                        set TotalProjectileTime = TotalProjectileTime + 0.04
                        if(ProjectileLimitTime<TotalProjectileTime): 
                            break
                    ProjectileProp.Dispose()
                    ObjetivePet.Heal(HealingPerk)
                    if(HexaMiniGameManager := MaybeHexaMiniGameManager?, BakeryMachinesManager := HexaMiniGameManager.MaybeBakeryMachinesManager?, Agent:=HexaMiniGameManager.MaybeAgent?):
                        BakeryMachinesManager.GameManager.VFXManager.Heal_AudioPlayer.Play(Agent)

    GenerateProjectile(ObjetivePet : hexagame_pet)<suspends> : void =
        if(ValidProp := MaybeProp?, ValidObjetivePet := ObjetivePet.MaybeProp?):

            ProjectileLimitTime:float=2.0
            var TotalProjectileTime:float = 0.0

            Tuple:=SpawnProp(RangeAttackEffectAsset,ValidProp.GetTransform())
            DistanceBetween:=Distance(ValidProp.GetTransform().Translation, ValidObjetivePet.GetTransform().Translation)
            TravelTime:=DistanceBetween/HexaGameProjectileSpeed
            

            FirstStageTime:=(TravelTime/10.0)*1.0
            MidStageTime:=(TravelTime/10.0)*8.0
            FinalStageTime:=(TravelTime/10.0)*1.0

            FirstStageTranslation := vector3{X:=ValidProp.GetTransform().Translation.X,Y:=ValidProp.GetTransform().Translation.Y,Z:=ValidProp.GetTransform().Translation.Z+HexaGameProjectileElevation}
            VectorToPoint1 := FirstStageTranslation-ValidProp.GetTransform().Translation

            MidStageTranslation := vector3{X:=ValidObjetivePet.GetTransform().Translation.X,Y:=ValidObjetivePet.GetTransform().Translation.Y,Z:=ValidObjetivePet.GetTransform().Translation.Z+HexaGameProjectileElevation}
            VectorToPoint2 := MidStageTranslation-(ValidProp.GetTransform().Translation+VectorToPoint1)

            FinalStageTranslation := vector3{X:=ValidObjetivePet.GetTransform().Translation.X,Y:=ValidObjetivePet.GetTransform().Translation.Y,Z:=ValidObjetivePet.GetTransform().Translation.Z}
            VectorToPoint3 := vector3{Z:=-2.0*HexaGameProjectileElevation}
            
            
            if:
                FinalTranslationX :=Floor[FinalStageTranslation.X]
                FinalTranslationY:=Floor[FinalStageTranslation.Y]
            then:

                <#
                #Print("FinalTranslationY: {FinalTranslationY}, start translation Y {ValidProp.GetTransform().Translation.Y}")
                FirstStageTranslation := vector3{X:=ValidProp.GetTransform().Translation.X,Y:=ValidProp.GetTransform().Translation.Y,Z:=ValidProp.GetTransform().Translation.Z+HexaGameProjectileElevation}
                VectorToPoint1 := FirstStageTranslation-ValidProp.GetTransform().Translation

                MidStageTranslation := vector3{X:=ValidObjetivePet.GetTransform().Translation.X,Y:=ValidObjetivePet.GetTransform().Translation.Y,Z:=ValidObjetivePet.GetTransform().Translation.Z+HexaGameProjectileElevation}
                var VectorToPoint2 :vector3 = MidStageTranslation-(ValidObjetivePet.GetTransform().Translation+VectorToPoint1)

                var VectorToPoint3 : vector3 = vector3{Z:=-2.0*HexaGameProjectileElevation}#>


                ##Print("FirstStageTime: {FirstStageTime} | MidStageTime: {MidStageTime} | FinalStageTime: {FinalStageTime}")

                

                if(ProjectileProp := Tuple(0)?):
                    MovementKeyFrame1 : keyframe_delta = keyframe_delta #Key Frame 1 to begining
                    {
                        DeltaLocation := VectorToPoint1
                        DeltaScale := NeutralScale
                        DeltaRotation := rotation{}
                        Time := FirstStageTime
                        Interpolation := EaseOut
                    }

                    MovementKeyFrame2 : keyframe_delta = keyframe_delta #Key Frame 1 to begining
                    {
                        DeltaLocation := VectorToPoint2
                        DeltaScale := NeutralScale
                        DeltaRotation := rotation{}
                        Time := MidStageTime
                        Interpolation := EaseIn
                        
                    }

                    MovementKeyFrame3 : keyframe_delta = keyframe_delta #Key Frame 1 to begining
                    {
                        DeltaLocation := VectorToPoint3
                        DeltaScale := NeutralScale
                        DeltaRotation := rotation{}
                        Time := FinalStageTime
                        Interpolation := EaseIn
                    }


                    if(var ProjectilePropController : animation_controller = ProjectileProp.GetAnimationController[]):
                        # Make an array of key frames
                        FortVaderKeyFrames :  []keyframe_delta = array { MovementKeyFrame1 , MovementKeyFrame2,  MovementKeyFrame3}
                        # Set the animation to play
                        ProjectilePropController.SetAnimation(FortVaderKeyFrames, ?Mode:=animation_mode.OneShot)
                        # Play Animation
                        var MidPointReached : logic = false
                        ##Print("Projectile launched and moved")
                        ProjectilePropController.Play()
                        #f(ValidEffectSound := MaybeEffectSound?):
                        #    ValidEffectSound.Play()
                        loop:
                            State := ProjectilePropController.GetState()
                            if(State = animation_controller_state.Stopped):
                                ##Print("Projectile reached destiny")
                                ProjectileProp.Dispose()
                                break
                            Sleep(0.04)
                            set TotalProjectileTime = TotalProjectileTime + 0.04
                            if(TotalProjectileTime>(TravelTime/2.0)):
                                #Check if objetiveMoved
                                if:
                                    CurrentX:=Floor[ValidObjetivePet.GetTransform().Translation.X]
                                    CurrentY:=Floor[ValidObjetivePet.GetTransform().Translation.Y]
                                then:
                                    <#
                                    if(CurrentX <> FinalTranslationX or CurrentY <> FinalTranslationY):
                                        #Print("Objetive has moved, recalculating...")
                                        ProjectilePropController.Stop()
                                        set VectorToPoint3 = ValidObjetivePet.GetTransform().Translation - ProjectileProp.GetTransform().Translation
                                        if(Controller := ProjectileProp.GetAnimationController[], set ProjectilePropController = Controller):
                                            # Set the animation to play
                                            ProjectilePropController.SetAnimation(array{MovementKeyFrame2}, ?Mode:=animation_mode.OneShot)
                                            ProjectilePropController.Play()#>
                            if(ProjectileLimitTime<TotalProjectileTime):
                                break
                    ProjectileProp.Dispose()
                    ObjetivePet.DealDamage(AttackDamage)
                    if(HexaMiniGameManager := MaybeHexaMiniGameManager?, BakeryMachinesManager := HexaMiniGameManager.MaybeBakeryMachinesManager?, Agent:=HexaMiniGameManager.MaybeAgent?):
                        BakeryMachinesManager.GameManager.VFXManager.Damage_AudioPlayer.Play(Agent)
                
    AnimateMovement(IsSuperJump : logic, FinalTransform : transform, FinalRotation : rotation)<suspends> : void =
        Tolerance : float = 5.0 #cm

        if(ValidProp := MaybeProp?, ValidProp.IsValid[]):

            InitialTransform := ValidProp.GetTransform()

            #Pitch Or Roll calculator

            var StartRotation : rotation = rotation{}
            var EndRotation : rotation = rotation{}

            var StartElevation : float = 16.0
            var EndElevation : float = -16.0

            if(IsSuperJump?):
                set StartElevation = 100.0
                set EndElevation = -100.0

            RotationDegrees : float = 0.436

            SpeedFrame0 : float = (Speed/8.0)*1
            SpeedFrame1 : float = (Speed/8.0)*3
            SpeedFrame2 : float = (Speed/8.0)*3
            SpeedFrame3 : float = (Speed/8.0)*1

            NewTransform : transform = transform{
                    Translation:=ValidProp.GetTransform().Translation,
                    Rotation:=FinalRotation,
                    Scale:=NormalScale}

            ValidProp.MoveTo(NewTransform,SpeedFrame0)
            
            if(-Tolerance<=ValidProp.GetTransform().Translation.Z-FinalTransform.Translation.Z<=Tolerance):
                ##Print("Z final and Z start are equal. Elevation will not be applied")
            else if(ValidProp.GetTransform().Translation.Z-FinalTransform.Translation.Z<Tolerance):
                ##Print("The token will rise")
                set StartElevation = StartElevation + (FinalTransform.Translation.Z - ValidProp.GetTransform().Translation.Z)
                set EndElevation = EndElevation + (ValidProp.GetTransform().Translation.Z - FinalTransform.Translation.Z)

            #Rotation calculator
            if(-Tolerance<=ValidProp.GetTransform().Translation.Y-FinalTransform.Translation.Y<=Tolerance):
                ##Print("Start coord Y and End coord Y are equal. Roll will not be applied")
            else if(ValidProp.GetTransform().Translation.Y-FinalTransform.Translation.Y<Tolerance):
                set StartRotation = InitialTransform.Rotation.ApplyRoll(-RotationDegrees)
                set EndRotation = InitialTransform.Rotation.ApplyRoll(RotationDegrees) 
            else:
                set StartRotation = InitialTransform.Rotation.ApplyRoll(RotationDegrees)
                set EndRotation = InitialTransform.Rotation.ApplyRoll(-RotationDegrees) 

            if(-Tolerance<=ValidProp.GetTransform().Translation.X-FinalTransform.Translation.X<=Tolerance):
                ##Print("Start coord X and End coord X are equal. Pith will not be applied")
            else if(ValidProp.GetTransform().Translation.X-FinalTransform.Translation.X<Tolerance):
                set StartRotation = InitialTransform.Rotation.ApplyPitch(RotationDegrees)
                set EndRotation = InitialTransform.Rotation.ApplyPitch(-RotationDegrees)
            else:
                set StartRotation = InitialTransform.Rotation.ApplyPitch(-RotationDegrees)
                set EndRotation = InitialTransform.Rotation.ApplyPitch(RotationDegrees)

            ##Print("Objetive position of the token \nX: {FinalTransform.Translation.X}\nY: {FinalTransform.Translation.Y}\nZ: {FinalTransform.Translation.Z}")

            MovementKeyFrame1 : keyframe_delta = keyframe_delta #Key Frame 1
            {
                DeltaLocation := vector3{X:=0.0,Y:=0.0,Z:=StartElevation}
                DeltaRotation := rotation{}
                DeltaScale := NeutralScale
                Time := SpeedFrame1
                Interpolation := EaseIn
            }

            MovementKeyFrame2 : keyframe_delta = keyframe_delta #Key Frame 2
            {
                DeltaLocation := vector3{X:=FinalTransform.Translation.X,Y:=FinalTransform.Translation.Y,Z:=FinalTransform.Translation.Z} - vector3{X:=InitialTransform.Translation.X,Y:=InitialTransform.Translation.Y,Z:=InitialTransform.Translation.Z}
                DeltaRotation := rotation{}
                DeltaScale := NeutralScale
                Time := SpeedFrame2
                Interpolation := Linear
            }

            MovementKeyFrame3 : keyframe_delta = keyframe_delta #Key Frame 3
            {
                DeltaLocation := vector3{X:=0.0,Y:=0.0,Z:=EndElevation}
                DeltaRotation := rotation{}
                DeltaScale := NeutralScale
                Time := SpeedFrame3
                Interpolation := EaseOut
            }

            if(PropController := ValidProp.GetAnimationController[]):
                # Make an array of key frames
                FortVaderKeyFrames :  []keyframe_delta = array { MovementKeyFrame1 , MovementKeyFrame2,  MovementKeyFrame3}
                # Set the animation to play
                PropController.SetAnimation(FortVaderKeyFrames, ?Mode:=animation_mode.OneShot)
                # Play Animation
                PropController.Play()
                #for(Player : GetPlayspace().GetPlayers(), Agent := agent[Player]):
                #    TokenMove_AudioPlayer.Play(Agent)
                loop:
                    State := PropController.GetState()
                    if(State = animation_controller_state.Stopped):
                        if(HexaMiniGameManager := MaybeHexaMiniGameManager?, BakeryMachinesManager := HexaMiniGameManager.MaybeBakeryMachinesManager?, Agent:=HexaMiniGameManager.MaybeAgent?):
                            BakeryMachinesManager.GameManager.VFXManager.HexaPetMove_AudioPlayer.Play(Agent)
                        if(Health<=0.0):
                            PlayDieEffect()
                        break
                    Sleep(0.04)
                set IsMoving = false
           

    DespawnEffect(EffectProp : creative_prop)<suspends> : void =
        Sleep(2.0)
        EffectProp.Dispose()

    SetInitialPosition(NewPosition : tuple(int,int)) : void =
        set LastPosition = Position
        set Position = NewPosition

    SetNewPosition(NewPosition : tuple(int,int)) : void =
        if(NewPosition(0)<>-1, NewPosition(1)<>-1, ValidProp := MaybeProp?):
            
            if:
                HexaMiniGameManager := MaybeHexaMiniGameManager?
                HexaMiniGameManager.CheckIfPositionIsEmpty(NewPosition)?
                set HexaMiniGameManager.AllHexBoxes[NewPosition(0)][NewPosition(1)]=option{Self}
                set HexaMiniGameManager.AllHexBoxes[Position(0)][Position(1)]=false
                OldHexProp := HexaMiniGameManager.AllHexPositions[Position(0)][Position(1)]
                HexProp := HexaMiniGameManager.AllHexPositions[NewPosition(0)][NewPosition(1)]
            then:
                HexTransform:=HexProp.GetTransform()
                Translation : vector3 = HexTransform.Translation+PivotVariation
                NewTransform:=transform{Translation:=Translation,Rotation:=HexProp.GetTransform().Rotation,Scale:=NormalScale}

                CurrentTranslation := ValidProp.GetTransform().Translation

                ## New code
                set LastPosition = Position
                set Position = NewPosition
                set AllMovementIntentions = array{} #Why is this here? it make no sense to put the last position BEFORE moving the prop!
                ##
                set IsMoving = true
                spawn:
                    MoveProp(Translation,CurrentTranslation)
            


    SearchObjetive(IsFindingEnemies : logic) : logic =
        var ActionMade : logic = false
        var AllObjetives : []hexagame_pet = array{}
        var Range : int = 0
        #Find an Enemy and Friendlies
        FindFriendliesAndFoes()
        if(IsFindingEnemies?):
            set Range = AttackRange
            set AllObjetives = AllEnemiesFound
        else:
            set Range = DefendRange
            set AllObjetives = AllFriendliesFound

        var DistancePosibleConfirmed_Objetive : int = 9999999999999
        var PosibleConfirmed_Objetive : ?hexagame_pet = false

        if(HexaMiniGameManager := MaybeHexaMiniGameManager?, ValidProp := MaybeProp?):
            PathTuple:=HexaMiniGameManager.PathFinder(Self,AllObjetives,Range,IsFindingEnemies)
            if(0<PathTuple(1)): #Use path finder is there is posible movement nearby (625 possibilities)
                #Print("Using new code")
                if(NextMove := PathTuple(0)[0]):
                    #Print("{NextMove(0)}/{NextMove(1)}")
                    AddLast_AllMovementIntentions(NextMove)
                    Move()
            else: 
                #Print("Using old code")
                #Use old code
                if(IsFindingEnemies?):
                    for(Objetive : AllObjetives, ValidObjetivePet := Objetive.MaybeProp?):
                        #ATTACK
                        if(Objetive.Health > 0.0, DistanceBetweenPets:=Floor[DistanceXY(ValidObjetivePet.GetTransform().Translation,ValidProp.GetTransform().Translation)]):
                            if(ValidObjetive := PosibleConfirmed_Objetive?, ValidObjetive.Health > 0.0):
                                if(DistanceBetweenPets+2<DistancePosibleConfirmed_Objetive): #If distance is shorter, assing this new enemy as objetive
                                    set PosibleConfirmed_Objetive = option{Objetive}
                                    set DistancePosibleConfirmed_Objetive = DistanceBetweenPets
                                else if((DistanceBetweenPets-DistancePosibleConfirmed_Objetive)<2, ValidObjetive.Health > Objetive.Health):  #If distance is the same, assign the weaker objetive
                                    set PosibleConfirmed_Objetive = option{Objetive}
                                    set DistancePosibleConfirmed_Objetive = DistanceBetweenPets
                            else: #There is no objetive assinged yet let´s assign this new one 
                                set PosibleConfirmed_Objetive = option{Objetive}
                                set DistancePosibleConfirmed_Objetive = DistanceBetweenPets
                else:  
                    for(Objetive : AllObjetives, ValidObjetivePet := Objetive.MaybeProp?):
                        #DEFEND
                        if(Objetive.Health > 0.0, DistanceBetweenPets:=Floor[DistanceXY(ValidObjetivePet.GetTransform().Translation,ValidProp.GetTransform().Translation)]):
                            if(ValidObjetive := PosibleConfirmed_Objetive?, ValidObjetive.Health > 0.0):
                                if(ValidObjetive.PetType = DRUID, Objetive.PetType <> DRUID): 
                                    #If the valid is support and the new no, replace
                                    set PosibleConfirmed_Objetive = option{Objetive}
                                    set DistancePosibleConfirmed_Objetive = DistanceBetweenPets
                                else if(Objetive.PetType <> DRUID):
                                    #If the new is not a support compare
                                    if(DistanceBetweenPets<DistancePosibleConfirmed_Objetive): #If distance is shorter, assing this new enemy as objetive
                                        set PosibleConfirmed_Objetive = option{Objetive}
                                        set DistancePosibleConfirmed_Objetive = DistanceBetweenPets
                                    else if(DistanceBetweenPets=DistancePosibleConfirmed_Objetive, ValidObjetive.Health > Objetive.Health):  #If distance is the same, assign the weaker objetive
                                        set PosibleConfirmed_Objetive = option{Objetive}
                                        set DistancePosibleConfirmed_Objetive = DistanceBetweenPets
                                else if(ValidObjetive.PetType = DRUID, Objetive.PetType = DRUID):
                                    #If both are support, compare
                                    if(DistanceBetweenPets<DistancePosibleConfirmed_Objetive): #If distance is shorter, assing this new enemy as objetive
                                        set PosibleConfirmed_Objetive = option{Objetive}
                                        set DistancePosibleConfirmed_Objetive = DistanceBetweenPets
                                    else if(DistanceBetweenPets=DistancePosibleConfirmed_Objetive, ValidObjetive.Health > Objetive.Health):  #If distance is the same, assign the weaker objetive
                                        set PosibleConfirmed_Objetive = option{Objetive}
                                        set DistancePosibleConfirmed_Objetive = DistanceBetweenPets
                            else: #There is no objetive assinged yet let´s assign this new one
                                set PosibleConfirmed_Objetive = option{Objetive}
                                set DistancePosibleConfirmed_Objetive = DistanceBetweenPets
                if:
                    ValidObjetive := PosibleConfirmed_Objetive?
                then:
                    if:
                        DistancePosibleConfirmed_Objetive > Range
                    then:
                        ##Print("Too far! Distance: {DistancePosibleConfirmed_Objetive}, Range: {Range}")
                        GenerateMovementIntention(ValidObjetive) #If there is a valid Objetive, but is too far, lets move
                        Move()
                    else:
                        ##Print("On range! Distance: {DistancePosibleConfirmed_Objetive}, Range: {Range}")
                        #If there is a valid enemy nearby, don´t move and continue attacking
                        AttackOfDefend(IsFindingEnemies,ValidObjetive)
                        set ActionMade = true
                        return ActionMade
                else:
                    #Print("FATAL ERROR. OBJETIVE NOT FOUND")
                    return ActionMade
            return ActionMade
        else:
            return ActionMade

    GenerateMovementIntention(PetObjetive : hexagame_pet) : void =
        #if(PetType=TANK):
        #    #Print("Generating movement intention")
        #Set the move intention 
        if(HexaMiniGameManager := MaybeHexaMiniGameManager?, ValidObjetivePet := PetObjetive.MaybeProp?):
            #Print("PetObjetive detected")
            BoxDistance := HexaMiniGameManager.CalculeBoxDistance(GetPosition(),PetObjetive.GetPosition())

            RowPosition := GetPosition()(0)
            ColumnPosition := GetPosition()(1)

            var TupleRowMoveOptions : tuple(tuple(int,int),tuple(int,int)) = ((-1,-1),(-1,-1))
            var TupleColumnMoveOption : tuple(int,int) = (-1,-1)

            #### ROW MOVE
            if(BoxDistance(0) > 1): #ROW #If distance is greater than 1 box, intention to move forward or backwards
                ##Print("Box distance is [{BoxDistance(0)}][{BoxDistance(1)}]")
                if(Position(0)>PetObjetive.Position(0)): 
                    set TupleRowMoveOptions=HexaMiniGameManager.GetNearDownPositions(GetPosition()) #Go down if objetive is under pet
                else if(Position(0)<PetObjetive.Position(0)): 
                    set TupleRowMoveOptions=HexaMiniGameManager.GetNearUpPositions(GetPosition()) #Go up if objetive is up pet
                ##Print("Objetive is down. \nMoving to [{TupleRowMoveOptions(0)(0)}][{TupleRowMoveOptions(0)(1)}] or to [{TupleRowMoveOptions(1)(0)}][{TupleRowMoveOptions(1)(1)}]")
                #Now check the correct column position from both option
                if:
                    HexaPropOption0:=HexaMiniGameManager.AllHexPositions[TupleRowMoveOptions(0)(0)][TupleRowMoveOptions(0)(1)]
                    HexaPropOption1:=HexaMiniGameManager.AllHexPositions[TupleRowMoveOptions(1)(0)][TupleRowMoveOptions(1)(1)]
                then:
                    DistanceOption0:=Distance(ValidObjetivePet.GetTransform().Translation,HexaPropOption0.GetTransform().Translation)
                    DistanceOption1:=Distance(ValidObjetivePet.GetTransform().Translation,HexaPropOption1.GetTransform().Translation)
                    if(DistanceOption0>DistanceOption1):
                        ##Print("Added to movement intentions Option 1")
                        AddLast_AllMovementIntentions(TupleRowMoveOptions(1))
                        AddLast_AllMovementIntentions(TupleRowMoveOptions(0))
                        
                    else:
                        ##Print("Added to movement intentions Option 0")
                        AddLast_AllMovementIntentions(TupleRowMoveOptions(0))
                        AddLast_AllMovementIntentions(TupleRowMoveOptions(1))              
            #### COLUMN MOVE
            else if(BoxDistance(1) > 1): #If distance more than 1 try to move on the row
                ##Print("Box distance is [{BoxDistance(0)}][{BoxDistance(1)}]")
                if(Position(1)<PetObjetive.Position(1)): #Objetive is on the right side
                    set TupleColumnMoveOption=HexaMiniGameManager.GetNearSidePositions(GetPosition())(1) 
                    #Print("Movement intention added")
                    AddFirst_AllMovementIntentions(TupleColumnMoveOption)
                    #Extra options!
                    if(Position(0)<PetObjetive.Position(0)): 
                        #Objetive is top the row
                        TupleExtraOptions:=HexaMiniGameManager.GetNearUpPositions(GetPosition()) #Go up if objetive is up pet
                        #Print("Movement intention added")
                        AddLast_AllMovementIntentions(TupleExtraOptions(1)) #Get right top if it is empty
                        AddLast_AllMovementIntentions(TupleExtraOptions(0)) #Get left top if it is empty
                    else if(Position(0)>PetObjetive.Position(0)):
                        #Objetive is under the row
                        TupleExtraOptions:=HexaMiniGameManager.GetNearDownPositions(GetPosition()) #Go up if objetive is up pet
                        #Print("Movement intention added")
                        AddLast_AllMovementIntentions(TupleExtraOptions(1)) #Get right under if it is empty
                        AddLast_AllMovementIntentions(TupleExtraOptions(0)) #Get left under if it is empty
                    else:
                        #Objetive is in the same row. advance depending on the player faction
                        if(IsPlayerPet = 0):
                            #Advance!
                            TupleExtraOptions:=HexaMiniGameManager.GetNearUpPositions(GetPosition()) #Go up if objetive is up pet
                            #Print("Movement intention added")
                            AddLast_AllMovementIntentions(TupleExtraOptions(1)) #Get the right option if it is empty
                            AddLast_AllMovementIntentions(TupleExtraOptions(0)) #Get the left option if it is empty
                        else:
                            #Go back!
                            TupleExtraOptions:=HexaMiniGameManager.GetNearDownPositions(GetPosition()) #Go up if objetive is up pet
                            #Print("Movement intention added")
                            AddLast_AllMovementIntentions(TupleExtraOptions(1)) #Get the right option if it is empty
                            AddLast_AllMovementIntentions(TupleExtraOptions(0)) #Get the left option if it is empty

                else if(Position(1)>PetObjetive.Position(1)): #Objetive is on the left side
                    set TupleColumnMoveOption=HexaMiniGameManager.GetNearSidePositions(GetPosition())(0) 
                    AddFirst_AllMovementIntentions(TupleColumnMoveOption)
                    #Print("Movement intention added")
                    #Extra options!
                    if(Position(0)<PetObjetive.Position(0)): 
                        #Objetive is top the row
                        TupleExtraOptions:=HexaMiniGameManager.GetNearUpPositions(GetPosition()) #Go up if objetive is up pet
                        AddLast_AllMovementIntentions(TupleExtraOptions(0)) #Get left top if it is empty
                        AddLast_AllMovementIntentions(TupleExtraOptions(1)) #Get right top if it is empty
                        #Print("Movement intention added")
                    else if(Position(0)>PetObjetive.Position(0)):
                        #Objetive is under the row
                        TupleExtraOptions:=HexaMiniGameManager.GetNearDownPositions(GetPosition()) #Go up if objetive is up pet
                        AddLast_AllMovementIntentions(TupleExtraOptions(0)) #Get left under if it is empty
                        AddLast_AllMovementIntentions(TupleExtraOptions(1)) #Get right under if it is empty
                        #Print("Movement intention added")
                    else:
                        #Objetive is in the same row. advance depending on the player faction
                        if(IsPlayerPet = 0):
                            #Advance!
                            TupleExtraOptions:=HexaMiniGameManager.GetNearUpPositions(GetPosition()) #Go up if objetive is up pet
                            AddLast_AllMovementIntentions(TupleExtraOptions(0)) #Get the left option if it is empty
                            AddLast_AllMovementIntentions(TupleExtraOptions(1)) #Get the right option if it is empty
                            #Print("Movement intention added")
                        else:
                            #Go back!
                            TupleExtraOptions:=HexaMiniGameManager.GetNearDownPositions(GetPosition()) #Go up if objetive is up pet
                            AddLast_AllMovementIntentions(TupleExtraOptions(0)) #Get the left option if it is empty
                            AddLast_AllMovementIntentions(TupleExtraOptions(1)) #Get the right option if it is empty
                            #Print("Movement intention added")

            else if(BoxDistance(1) = 1) : #If distance is 1 try to move on the row only in special cases 
                if(Position(0)<PetObjetive.Position(0)): #Objetive is on top
                    if(Position(1)<PetObjetive.Position(1)): #Objetive is on the right side
                        TupleExtraOption:=HexaMiniGameManager.GetNearUpPositions(GetPosition()) #Objetive is on the right side
                        ExtraOption:=HexaMiniGameManager.GetNearSidePositions(GetPosition())(1) #Objetive is on the right side
                        #Print("Movement intention added")
                        AddLast_AllMovementIntentions(TupleExtraOption(1))
                        AddLast_AllMovementIntentions(ExtraOption) #Move be near
                        AddLast_AllMovementIntentions(TupleExtraOption(0)) #Move to open space
                    else if(Position(1)>PetObjetive.Position(1)):  #Objetive is on the left side
                        TupleExtraOption:=HexaMiniGameManager.GetNearUpPositions(GetPosition()) #Objetive is on the left side
                        ExtraOption:=HexaMiniGameManager.GetNearSidePositions(GetPosition())(0) #Objetive is on the left side
                        #Print("Movement intention added")
                        AddLast_AllMovementIntentions(TupleExtraOption(0))
                        AddLast_AllMovementIntentions(ExtraOption) #Move be near
                        AddLast_AllMovementIntentions(TupleExtraOption(1)) #Move to open space
                else if(Position(0)>PetObjetive.Position(0)): #Objetive under on top
                    if(Position(1)<PetObjetive.Position(1)): #Objetive is on the right side
                        TupleExtraOption:=HexaMiniGameManager.GetNearDownPositions(GetPosition()) #Objetive is on the right side
                        ExtraOption:=HexaMiniGameManager.GetNearSidePositions(GetPosition())(1) #Objetive is on the right side
                        #Print("Movement intention added")
                        AddLast_AllMovementIntentions(TupleExtraOption(1))
                        AddLast_AllMovementIntentions(ExtraOption) #Move be near
                        AddLast_AllMovementIntentions(TupleExtraOption(0)) #Move to open space
                    else if(Position(1)>PetObjetive.Position(1)):  #Objetive is on the left side
                        TupleExtraOption:=HexaMiniGameManager.GetNearDownPositions(GetPosition()) #Objetive is on the left side
                        ExtraOption:=HexaMiniGameManager.GetNearSidePositions(GetPosition())(0) #Objetive is on the left side
                        #Print("Movement intention added")
                        AddLast_AllMovementIntentions(TupleExtraOption(0))
                        AddLast_AllMovementIntentions(ExtraOption) #Move be near
                        AddLast_AllMovementIntentions(TupleExtraOption(1)) #Move to open space 
                        
        ##Print("AllMovementIntentions.Length: {AllMovementIntentions.Length}")

    ResetHealth() : void =
        set Health = MaxHealth
        set IsDeath = false

    DealDamage(Ammount : float) : void =
        OldHealth := Health
        if(Health>0.0):
            set Health = Health - Ammount
            if(Health<=0.0):
                set Health = 0.0
        if(HexaMiniGameManager := MaybeHexaMiniGameManager?):
            if(Health<=0.0):
                set Health = 0.0
                #Die()
                #HexaMiniGameManager.CheckGameEnd(false)
            HexaMiniGameManager.RefreshHexagameUI()
        spawn:
            HealthChange(OldHealth,Health)

    Heal(Ammount : float) : void =
        if(Health>0.0):
            if(HexaMiniGameManager := MaybeHexaMiniGameManager?):
                OldHealth:=Health
                set Health = Health + Ammount 
                if(MaxHealth<Health):
                    set Health = MaxHealth
                HexaMiniGameManager.RefreshHexagameUI()
                spawn:
                    HealthChange(OldHealth,Health)

    HealthChange(OldHealth : float, NewHealth  : float)<suspends>: void =
        if:
            OldBarFragment:=Ceil[(OldHealth/MaxHealth)*10.0]
            NewBarFragment:=Ceil[(NewHealth/MaxHealth)*10.0]
            HexaMiniGameManager := MaybeHexaMiniGameManager?
            ValidProp := MaybeProp?
        then:
            if(OldBarFragment>NewBarFragment):
                for(BarPosition:=NewBarFragment+3..OldBarFragment+2, Prop := MaybeProp?):
                    Prop.SetMaterial(YellowPetBarMaterial,?Index:=BarPosition)
                Sleep(0.5)
                SetHealthUI(ValidProp,MaxHealth,Health,IsPlayerPet)
            
            else if(NewBarFragment>OldBarFragment):
                for(BarPosition:=OldBarFragment+3..NewBarFragment+2, Prop := MaybeProp?):
                        Prop.SetMaterial(BluePetBarMaterial,?Index:=BarPosition)
                Sleep(0.5)
                SetHealthUI(ValidProp,MaxHealth,Health, IsPlayerPet)
        else:
            PrintFatalError("On HealthChange integer overflow")

    Die() : void =
        if(not IsDeath?):
            set IsDeath = true
            set Health = 0.0
            if:
                HexaMiniGameManager := MaybeHexaMiniGameManager?
                set HexaMiniGameManager.AllHexBoxes[Position(0)][Position(1)]=false
            then:
                HexaMiniGameManager.CheckGameEnd(false)
                if(HexProp := HexaMiniGameManager.AllHexPositions[Position(0)][Position(1)]):
                    HexProp.SetMaterial(Materials.MI_HexaBaseNoFresnel, ?Index:=2)
                spawn:
                    PlayDieEffect()
            
        
    PlayDieEffect()<suspends> : void =
        if(ValidProp := MaybeProp?, ValidProp.IsValid[]):
            if(IsMoving?):
                Sleep(0.5)
            PropTranslation:=ValidProp.GetTransform().Translation
            PropRotation:=ValidProp.GetTransform().Rotation
            SmallScale:=vector3{X:=0.1,Y:=0.1,Z:=0.1}
            NewTransform:=transform{Translation:=vector3{X:=PropTranslation.X,Y:=PropTranslation.Y,Z:=PropTranslation.Z-128.0},
                Rotation:=PropRotation, Scale:=SmallScale}
            ValidProp.MoveTo(NewTransform,1.0)
            set AllFriendliesFound = array{}
            set AllEnemiesFound = array{}



    GetPosition() : tuple(int,int) =
        return Position

    GetType() : int =
        return PetType

    AddLast_AllMovementIntentions (NewPosition : tuple(int,int)) : void = 
        set AllMovementIntentions = AllMovementIntentions + array{NewPosition}

    AddFirst_AllMovementIntentions (NewPosition : tuple(int,int)) : void = 
        set AllMovementIntentions = array{NewPosition} + AllMovementIntentions

    GetFaction() : logic =
        return IsPlayerPet

    #Setters
    SetNewPetStats(InHexaMiniGameManager : hexa_minigame_manager, InPropAsset : creative_prop_asset, InPackA_LV : int, InPackB_LV : int,InMaxHealth : float, InAttackDamage : float, InAttackRange : int, InDefendRange : int, InPetType : int, InIsPlayerPet : logic, InTier : int, InPurity : int, InSpeed : float, InActionSpeed : float, InHealingPerk : float) : void =
        set MaybeHexaMiniGameManager = option{InHexaMiniGameManager}
        set PropAsset = InPropAsset
        set MaxHealth = InMaxHealth
        set PackA_LV = InPackA_LV
        set PackB_LV = InPackB_LV
        set Health = MaxHealth
        set AttackDamage = InAttackDamage
        set AttackRange = InAttackRange
        set DefendRange = InDefendRange
        set PetType = InPetType
        set IsPlayerPet = InIsPlayerPet
        set Tier = InTier
        set Purity = InPurity
        set Speed = InSpeed
        set ActionSpeed = InActionSpeed
        set HealingPerk = InHealingPerk
        

MakeHexaPetTable<constructor>(
    InAllPetsIndex : int,
    InHexaMiniGameManager : ?hexa_minigame_manager,
    InPropAsset : creative_prop_asset,
    InMeleeAttackEffectAsset : creative_prop_asset,
    InRangeAttackEffectAsset : creative_prop_asset,
    InDefendEffectAsset : creative_prop_asset,
    InPackA_LV : int,
    InPackB_LV : int,
    InAttackRange : int,
    InDefendRange : int,
    InPetType : int,
    InIsPlayerPet : logic,
    InTier : int,
    InPurity : int,
    InMaxHealth : float,
    InAttackDamage : float,
    InSpeed : float,
    InActionSpeed : float,
    InHealingPerk : float 
    )<transacts> := hexagame_pet :
        AllPetsIndex := InAllPetsIndex
        MaybeHexaMiniGameManager := InHexaMiniGameManager
        PropAsset := InPropAsset
        MeleeAttackEffectAsset := InMeleeAttackEffectAsset
        RangeAttackEffectAsset := InRangeAttackEffectAsset
        HealEffectAsset := InDefendEffectAsset
        PackA_LV := InPackA_LV
        PackB_LV := InPackB_LV
        AttackRange := InAttackRange
        DefendRange := InDefendRange
        PetType := InPetType
        IsPlayerPet := InIsPlayerPet
        Tier := InTier
        Purity := InPurity
        MaxHealth := InMaxHealth
        Health := InMaxHealth
        AttackDamage := InAttackDamage
        Speed := InSpeed
        ActionSpeed := InActionSpeed
        HealingPerk := InHealingPerk

        