using { /UnrealEngine.com/Temporary/UI }
using { /UnrealEngine.com/Temporary/SpatialMath }

#All challenges DataBase


hexa_challenges_database := class():
    
    var AllChallenges : [][]tuple(tuple(int,int,int,int),tuple(int,int)) = array{}
    var AllChallengesNames : []string = array{
        "Noob I"
        "Noob II"
        "Noob III"
        "Noob IV"
        "Cook I"
        "Cook II"
        "Cook III"
        "Cook IV"
        "Cook V"
        "Cook VI"
        "Cook VI"
        "Cook VII"
        "Cook VIII"
        "Master I"
        "Master II"
        "Master III"
        "Master IV"
        "Master V"
        "Master VI"
        "Master VII"
        "Master VIII"
        "Heissen I"
        "Heissen II"
        "Heissen III"
        "Heissen IV"
        "Heissen VV"

    }

    #1 ENEMY
    AllChallenge_Pets_0 : []tuple(tuple(int,int,int,int),tuple(int,int)) = 
        array{
            ((<#Type#>ADC,<#Purity#>0,<#PackA#>1,<#PackB#>1),<#Pos#>(5,6))
        }

    AllChallenge_Pets_0_1 : []tuple(tuple(int,int,int,int),tuple(int,int)) = 
        array{
            ((<#Type#>DRUID,<#Purity#>1,<#PackA#>1,<#PackB#>1),<#Pos#>(4,4))
        }

    AllChallenge_Pets_0_2 : []tuple(tuple(int,int,int,int),tuple(int,int)) = 
        array{
            ((<#Type#>TANK,<#Purity#>1,<#PackA#>1,<#PackB#>1),<#Pos#>(3,4))
        }

    AllChallenge_Pets_0_3 : []tuple(tuple(int,int,int,int),tuple(int,int)) = 
        array{
            ((<#Type#>HUNTER,<#Purity#>1,<#PackA#>1,<#PackB#>1),<#Pos#>(5,2))
        }

    AllChallenge_Pets_0_4 : []tuple(tuple(int,int,int,int),tuple(int,int)) = 
        array{
            ((<#Type#>FIGHTER,<#Purity#>1,<#PackA#>1,<#PackB#>1),<#Pos#>(3,4))
        }
    #2 ENEMIES
    AllChallenge_Pets_1 : []tuple(tuple(int,int,int,int),tuple(int,int)) = 
        array{
            C1P0 : tuple(tuple(int,int,int,int),tuple(int,int)) = ((<#Type#>DRUID,<#Purity#>3,<#PackA#>2,<#PackB#>2),<#Pos#>(4,6))
            C1P1 : tuple(tuple(int,int,int,int),tuple(int,int)) = ((<#Type#>DRUID,<#Purity#>3,<#PackA#>2,<#PackB#>2),<#Pos#>(5,6))
        }

    AllChallenge_Pets_1_1 : []tuple(tuple(int,int,int,int),tuple(int,int)) = 
        array{
            ((<#Type#>FIGHTER,<#Purity#>3,<#PackA#>2,<#PackB#>2),<#Pos#>(4,6))
            ((<#Type#>DRUID,<#Purity#>3,<#PackA#>2,<#PackB#>2),<#Pos#>(5,6))
        }

    AllChallenge_Pets_1_2 : []tuple(tuple(int,int,int,int),tuple(int,int)) = 
        array{
            ((<#Type#>FIGHTER,<#Purity#>3,<#PackA#>3,<#PackB#>3),<#Pos#>(4,6))
            ((<#Type#>DRUID,<#Purity#>3,<#PackA#>3,<#PackB#>3),<#Pos#>(5,6))
        }

    AllChallenge_Pets_1_3 : []tuple(tuple(int,int,int,int),tuple(int,int)) = 
        array{
            ((<#Type#>HUNTER,<#Purity#>3,<#PackA#>3,<#PackB#>3),<#Pos#>(5,2))
            ((<#Type#>DRUID,<#Purity#>3,<#PackA#>3,<#PackB#>3),<#Pos#>(5,6))
        }

    AllChallenge_Pets_1_4 : []tuple(tuple(int,int,int,int),tuple(int,int)) = 
        array{
            ((<#Type#>TANK,<#Purity#>5,<#PackA#>3,<#PackB#>3),<#Pos#>(3,3))
            ((<#Type#>DRUID,<#Purity#>3,<#PackA#>3,<#PackB#>3),<#Pos#>(5,5))
        }

    AllChallenge_Pets_1_5 : []tuple(tuple(int,int,int,int),tuple(int,int)) = 
        array{
            ((<#Type#>HUNTER,<#Purity#>4,<#PackA#>5,<#PackB#>5),<#Pos#>(5,2))
            ((<#Type#>HUNTER,<#Purity#>4,<#PackA#>5,<#PackB#>5),<#Pos#>(5,3))
        }
    #3 ENEMIES
    AllChallenge_Pets_2 : []tuple(tuple(int,int,int,int),tuple(int,int)) = 
        array{
            C2P0 : tuple(tuple(int,int,int,int),tuple(int,int)) = ((<#Type#>FIGHTER,<#Purity#>5,<#PackA#>5,<#PackB#>5),<#Pos#>(4,6))
            C2P1: tuple(tuple(int,int,int,int),tuple(int,int)) = ((<#Type#>DRUID,<#Purity#>5,<#PackA#>5,<#PackB#>5),<#Pos#>(5,6))
            C2P2 : tuple(tuple(int,int,int,int),tuple(int,int)) = ((<#Type#>ADC,<#Purity#>5,<#PackA#>5,<#PackB#>5),<#Pos#>(5,4))
        }

    AllChallenge_Pets_2_1 : []tuple(tuple(int,int,int,int),tuple(int,int)) = 
        array{
            ((<#Type#>TANK,<#Purity#>5,<#PackA#>5,<#PackB#>5),<#Pos#>(3,3))
            ((<#Type#>DRUID,<#Purity#>5,<#PackA#>5,<#PackB#>5),<#Pos#>(4,4))
            ((<#Type#>ADC,<#Purity#>5,<#PackA#>5,<#PackB#>5),<#Pos#>(4,3))
        }

    AllChallenge_Pets_2_2 : []tuple(tuple(int,int,int,int),tuple(int,int)) = 
        array{
            ((<#Type#>TANK,<#Purity#>5,<#PackA#>5,<#PackB#>5),<#Pos#>(3,4))
            ((<#Type#>TANK,<#Purity#>5,<#PackA#>5,<#PackB#>5),<#Pos#>(3,5))
            ((<#Type#>ADC,<#Purity#>5,<#PackA#>5,<#PackB#>5),<#Pos#>(4,4))
        }

    AllChallenge_Pets_2_3 : []tuple(tuple(int,int,int,int),tuple(int,int)) = 
        array{
            ((<#Type#>HUNTER,<#Purity#>5,<#PackA#>5,<#PackB#>5),<#Pos#>(5,3))
            ((<#Type#>HUNTER,<#Purity#>5,<#PackA#>5,<#PackB#>5),<#Pos#>(5,4))
            ((<#Type#>HUNTER,<#Purity#>5,<#PackA#>5,<#PackB#>5),<#Pos#>(5,5))
        }

    AllChallenge_Pets_2_4 : []tuple(tuple(int,int,int,int),tuple(int,int)) = 
        array{
            ((<#Type#>FIGHTER,<#Purity#>5,<#PackA#>5,<#PackB#>5),<#Pos#>(3,3))
            ((<#Type#>FIGHTER,<#Purity#>5,<#PackA#>5,<#PackB#>5),<#Pos#>(3,4))
            ((<#Type#>FIGHTER,<#Purity#>5,<#PackA#>5,<#PackB#>5),<#Pos#>(3,5))
        }

    AllChallenge_Pets_2_5 : []tuple(tuple(int,int,int,int),tuple(int,int)) = 
        array{
            ((<#Type#>FIGHTER,<#Purity#>5,<#PackA#>5,<#PackB#>5),<#Pos#>(3,3))
            ((<#Type#>FIGHTER,<#Purity#>5,<#PackA#>5,<#PackB#>5),<#Pos#>(3,4))
            ((<#Type#>FIGHTER,<#Purity#>5,<#PackA#>5,<#PackB#>5),<#Pos#>(3,5))
        }

    #4 ENEMIES
    AllChallenge_Pets_3 : []tuple(tuple(int,int,int,int),tuple(int,int)) = 
        array{
            C3P0 : tuple(tuple(int,int,int,int),tuple(int,int)) = ((<#Type#>DRUID,<#Purity#>7,<#PackA#>5,<#PackB#>5),<#Pos#>(4,3))
            C3P1 : tuple(tuple(int,int,int,int),tuple(int,int)) = ((<#Type#>DRUID,<#Purity#>7,<#PackA#>5,<#PackB#>5),<#Pos#>(4,4))
            C3P2 : tuple(tuple(int,int,int,int),tuple(int,int)) = ((<#Type#>DRUID,<#Purity#>7,<#PackA#>5,<#PackB#>5),<#Pos#>(4,5))
            C3P3 : tuple(tuple(int,int,int,int),tuple(int,int)) = ((<#Type#>DRUID,<#Purity#>7,<#PackA#>5,<#PackB#>5),<#Pos#>(4,6))
        }


    AllChallenge_Pets_3_2 : []tuple(tuple(int,int,int,int),tuple(int,int)) = 
        array{
            ((<#Type#>ADC,<#Purity#>7,<#PackA#>5,<#PackB#>5),<#Pos#>(4,4))
            ((<#Type#>ADC,<#Purity#>10,<#PackA#>5,<#PackB#>5),<#Pos#>(4,5))
            ((<#Type#>ADC,<#Purity#>7,<#PackA#>5,<#PackB#>5),<#Pos#>(4,3))
            ((<#Type#>ADC,<#Purity#>7,<#PackA#>5,<#PackB#>5),<#Pos#>(3,5))
        }

    AllChallenge_Pets_3_3 : []tuple(tuple(int,int,int,int),tuple(int,int)) = 
        array{
            ((<#Type#>TANK,<#Purity#>15,<#PackA#>5,<#PackB#>5),<#Pos#>(3,4))
            ((<#Type#>ADC,<#Purity#>15,<#PackA#>5,<#PackB#>5),<#Pos#>(4,5))
            ((<#Type#>DRUID,<#Purity#>5,<#PackA#>5,<#PackB#>5),<#Pos#>(5,3))
            ((<#Type#>DRUID,<#Purity#>5,<#PackA#>5,<#PackB#>5),<#Pos#>(5,5))
        }

    AllChallenge_Pets_3_4 : []tuple(tuple(int,int,int,int),tuple(int,int)) = 
        array{
            ((<#Type#>TANK,<#Purity#>15,<#PackA#>5,<#PackB#>5),<#Pos#>(3,3))
            ((<#Type#>TANK,<#Purity#>10,<#PackA#>5,<#PackB#>5),<#Pos#>(3,4))
            ((<#Type#>TANK,<#Purity#>10,<#PackA#>5,<#PackB#>5),<#Pos#>(3,5))
            ((<#Type#>TANK,<#Purity#>5,<#PackA#>5,<#PackB#>5),<#Pos#>(3,6))
        }

    AllChallenge_Pets_3_5 : []tuple(tuple(int,int,int,int),tuple(int,int)) = 
        array{
            ((<#Type#>FIGHTER,<#Purity#>15,<#PackA#>5,<#PackB#>5),<#Pos#>(3,2))
            ((<#Type#>FIGHTER,<#Purity#>15,<#PackA#>5,<#PackB#>5),<#Pos#>(3,3))
            ((<#Type#>FIGHTER,<#Purity#>5,<#PackA#>5,<#PackB#>5),<#Pos#>(3,4))
            ((<#Type#>FIGHTER,<#Purity#>10,<#PackA#>5,<#PackB#>5),<#Pos#>(3,5))
        }

    AllChallenge_Pets_3_6 : []tuple(tuple(int,int,int,int),tuple(int,int)) = 
        array{
            ((<#Type#>HUNTER,<#Purity#>10,<#PackA#>5,<#PackB#>5),<#Pos#>(5,2))
            ((<#Type#>HUNTER,<#Purity#>10,<#PackA#>5,<#PackB#>5),<#Pos#>(5,3))
            ((<#Type#>HUNTER,<#Purity#>10,<#PackA#>5,<#PackB#>5),<#Pos#>(5,4))
            ((<#Type#>HUNTER,<#Purity#>10,<#PackA#>5,<#PackB#>5),<#Pos#>(5,5))
        }

    AllChallenge_Pets_3_7 : []tuple(tuple(int,int,int,int),tuple(int,int)) = 
        array{
            ((<#Type#>DRUID,<#Purity#>25,<#PackA#>7,<#PackB#>7),<#Pos#>(4,2))
            ((<#Type#>DRUID,<#Purity#>25,<#PackA#>7,<#PackB#>7),<#Pos#>(4,3))
            ((<#Type#>DRUID,<#Purity#>25,<#PackA#>7,<#PackB#>7),<#Pos#>(4,4))
            ((<#Type#>DRUID,<#Purity#>25,<#PackA#>7,<#PackB#>7),<#Pos#>(4,5))
        }

    #5 ENEMIES
    AllChallenge_Pets_4 : []tuple(tuple(int,int,int,int),tuple(int,int)) = 
        array{
            ((<#Type#>ADC,<#Purity#>15,<#PackA#>5,<#PackB#>5),<#Pos#>(4,3))
            ((<#Type#>ADC,<#Purity#>15,<#PackA#>5,<#PackB#>5),<#Pos#>(4,4))
            ((<#Type#>ADC,<#Purity#>15,<#PackA#>5,<#PackB#>5),<#Pos#>(4,5))
            ((<#Type#>ADC,<#Purity#>15,<#PackA#>5,<#PackB#>5),<#Pos#>(4,6))
            ((<#Type#>ADC,<#Purity#>15,<#PackA#>5,<#PackB#>5),<#Pos#>(5,4))
        }

    AllChallenge_Pets_5 : []tuple(tuple(int,int,int,int),tuple(int,int)) = 
        array{
            ((<#Type#>FIGHTER,<#Purity#>15,<#PackA#>5,<#PackB#>5),<#Pos#>(3,4))
            ((<#Type#>FIGHTER,<#Purity#>15,<#PackA#>5,<#PackB#>5),<#Pos#>(3,5))
            ((<#Type#>FIGHTER,<#Purity#>15,<#PackA#>5,<#PackB#>5),<#Pos#>(5,3))
            ((<#Type#>FIGHTER,<#Purity#>15,<#PackA#>5,<#PackB#>5),<#Pos#>(4,5))
            ((<#Type#>FIGHTER,<#Purity#>15,<#PackA#>5,<#PackB#>5),<#Pos#>(5,5))
        }

    AllChallenge_Pets_5_1 : []tuple(tuple(int,int,int,int),tuple(int,int)) = 
        array{
            ((<#Type#>TANK,<#Purity#>15,<#PackA#>7,<#PackB#>7),<#Pos#>(4,3))
            ((<#Type#>TANK,<#Purity#>15,<#PackA#>7,<#PackB#>7),<#Pos#>(4,4))
            ((<#Type#>TANK,<#Purity#>15,<#PackA#>7,<#PackB#>7),<#Pos#>(4,5))
            ((<#Type#>TANK,<#Purity#>15,<#PackA#>7,<#PackB#>7),<#Pos#>(4,6))
            ((<#Type#>TANK,<#Purity#>15,<#PackA#>7,<#PackB#>7),<#Pos#>(4,2))
        }

    #FROM HERE, GENERATE WITH CODE
    <#
    AllChallenge_Pets_5_1 : []tuple(tuple(int,int,int,int),tuple(int,int)) = 
        array{
            ((<#Type#>TANK,<#Purity#>15,<#PackA#>7,<#PackB#>7),<#Pos#>(4,3))
            ((<#Type#>TANK,<#Purity#>15,<#PackA#>7,<#PackB#>7),<#Pos#>(4,4))
            ((<#Type#>TANK,<#Purity#>15,<#PackA#>7,<#PackB#>7),<#Pos#>(4,5))
            ((<#Type#>TANK,<#Purity#>15,<#PackA#>7,<#PackB#>7),<#Pos#>(4,6))
            ((<#Type#>TANK,<#Purity#>15,<#PackA#>7,<#PackB#>7),<#Pos#>(4,2))
        }

    AllChallenge_Pets_6 : []tuple(tuple(int,int,int,int),tuple(int,int)) = 
        array{
            C6P0 : tuple(tuple(int,int,int,int),tuple(int,int)) = ((<#Type#>TANK,<#Purity#>15,<#PackA#>1,<#PackB#>1),<#Pos#>(3,4))
            C6P1 : tuple(tuple(int,int,int,int),tuple(int,int)) = ((<#Type#>FIGHTER,<#Purity#>15,<#PackA#>1,<#PackB#>1),<#Pos#>(3,5))
            C6P2 : tuple(tuple(int,int,int,int),tuple(int,int)) = ((<#Type#>HUNTER,<#Purity#>5,<#PackA#>1,<#PackB#>1),<#Pos#>(5,3))
            C6P3 : tuple(tuple(int,int,int,int),tuple(int,int)) = ((<#Type#>ADC,<#Purity#>5,<#PackA#>1,<#PackB#>1),<#Pos#>(4,5))
            C6P4 : tuple(tuple(int,int,int,int),tuple(int,int)) = ((<#Type#>DRUID,<#Purity#>5,<#PackA#>1,<#PackB#>1),<#Pos#>(5,5))
        }

    AllChallenge_Pets_7 : []tuple(tuple(int,int,int,int),tuple(int,int)) = 
        array{
            C7P0 : tuple(tuple(int,int,int,int),tuple(int,int)) = ((<#Type#>TANK,<#Purity#>25,<#PackA#>1,<#PackB#>1),<#Pos#>(3,4))
            C7P1 : tuple(tuple(int,int,int,int),tuple(int,int)) = ((<#Type#>FIGHTER,<#Purity#>25,<#PackA#>1,<#PackB#>1),<#Pos#>(3,3))
            C7P2 : tuple(tuple(int,int,int,int),tuple(int,int)) = ((<#Type#>HUNTER,<#Purity#>25,<#PackA#>1,<#PackB#>1),<#Pos#>(5,2))
            C7P3 : tuple(tuple(int,int,int,int),tuple(int,int)) = ((<#Type#>ADC,<#Purity#>15,<#PackA#>1,<#PackB#>1),<#Pos#>(4,5))
            C7P4 : tuple(tuple(int,int,int,int),tuple(int,int)) = ((<#Type#>DRUID,<#Purity#>15,<#PackA#>1,<#PackB#>1),<#Pos#>(5,6))
        }

    AllChallenge_Pets_8 : []tuple(tuple(int,int,int,int),tuple(int,int)) = 
        array{
            C8P0 : tuple(tuple(int,int,int,int),tuple(int,int)) = ((<#Type#>TANK,<#Purity#>35,<#PackA#>1,<#PackB#>1),<#Pos#>(3,4))
            C8P1 : tuple(tuple(int,int,int,int),tuple(int,int)) = ((<#Type#>FIGHTER,<#Purity#>35,<#PackA#>1,<#PackB#>1),<#Pos#>(3,3))
            C8P2 : tuple(tuple(int,int,int,int),tuple(int,int)) = ((<#Type#>HUNTER,<#Purity#>35,<#PackA#>1,<#PackB#>1),<#Pos#>(5,2))
            C8P3 : tuple(tuple(int,int,int,int),tuple(int,int)) = ((<#Type#>ADC,<#Purity#>45,<#PackA#>1,<#PackB#>1),<#Pos#>(4,5))
            C8P4 : tuple(tuple(int,int,int,int),tuple(int,int)) = ((<#Type#>DRUID,<#Purity#>25,<#PackA#>1,<#PackB#>1),<#Pos#>(5,6))
        }

    AllChallenge_Pets_9 : []tuple(tuple(int,int,int,int),tuple(int,int)) = 
        array{
            C9P0 : tuple(tuple(int,int,int,int),tuple(int,int)) = ((<#Type#>HUNTER,<#Purity#>45,<#PackA#>1,<#PackB#>1),<#Pos#>(5,1))
            C9P1 : tuple(tuple(int,int,int,int),tuple(int,int)) = ((<#Type#>HUNTER,<#Purity#>45,<#PackA#>1,<#PackB#>1),<#Pos#>(5,2))
            C9P2 : tuple(tuple(int,int,int,int),tuple(int,int)) = ((<#Type#>HUNTER,<#Purity#>25,<#PackA#>1,<#PackB#>1),<#Pos#>(5,3))
            C9P3 : tuple(tuple(int,int,int,int),tuple(int,int)) = ((<#Type#>HUNTER,<#Purity#>35,<#PackA#>1,<#PackB#>1),<#Pos#>(5,5))
            C9P4 : tuple(tuple(int,int,int,int),tuple(int,int)) = ((<#Type#>HUNTER,<#Purity#>25,<#PackA#>1,<#PackB#>1),<#Pos#>(5,6))
        }

    AllChallenge_Pets_10 : []tuple(tuple(int,int,int,int),tuple(int,int)) = 
        array{
            C10P0 : tuple(tuple(int,int,int,int),tuple(int,int)) = ((<#Type#>TANK,<#Purity#>45,<#PackA#>1,<#PackB#>1),<#Pos#>(5,1))
            C10P1 : tuple(tuple(int,int,int,int),tuple(int,int)) = ((<#Type#>ADC,<#Purity#>28,<#PackA#>1,<#PackB#>1),<#Pos#>(5,2))
            C10P2 : tuple(tuple(int,int,int,int),tuple(int,int)) = ((<#Type#>ADC,<#Purity#>75,<#PackA#>1,<#PackB#>1),<#Pos#>(5,3))
            C10P3 : tuple(tuple(int,int,int,int),tuple(int,int)) = ((<#Type#>ADC,<#Purity#>53,<#PackA#>1,<#PackB#>1),<#Pos#>(5,5))
            C10P4 : tuple(tuple(int,int,int,int),tuple(int,int)) = ((<#Type#>ADC,<#Purity#>44,<#PackA#>1,<#PackB#>1),<#Pos#>(5,6))
        }

    AllChallenge_Pets_11 : []tuple(tuple(int,int,int,int),tuple(int,int)) = 
        array{
            C11P0 : tuple(tuple(int,int,int,int),tuple(int,int)) = ((<#Type#>DRUID,<#Purity#>101,<#PackA#>1,<#PackB#>1),<#Pos#>(5,1))
            C11P1 : tuple(tuple(int,int,int,int),tuple(int,int)) = ((<#Type#>DRUID,<#Purity#>99,<#PackA#>1,<#PackB#>1),<#Pos#>(5,2))
            C11P2 : tuple(tuple(int,int,int,int),tuple(int,int)) = ((<#Type#>DRUID,<#Purity#>96,<#PackA#>1,<#PackB#>1),<#Pos#>(5,3))
            C11P3 : tuple(tuple(int,int,int,int),tuple(int,int)) = ((<#Type#>DRUID,<#Purity#>60,<#PackA#>1,<#PackB#>1),<#Pos#>(5,5))
            C11P4 : tuple(tuple(int,int,int,int),tuple(int,int)) = ((<#Type#>DRUID,<#Purity#>78,<#PackA#>1,<#PackB#>1),<#Pos#>(5,6))
        }

    AllChallenge_Pets_12 : []tuple(tuple(int,int,int,int),tuple(int,int)) = 
        array{
            C12P0 : tuple(tuple(int,int,int,int),tuple(int,int)) = ((<#Type#>ADC,<#Purity#>113,<#PackA#>10,<#PackB#>10),<#Pos#>(3,4))
            C12P1 : tuple(tuple(int,int,int,int),tuple(int,int)) = ((<#Type#>TANK,<#Purity#>134,<#PackA#>10,<#PackB#>10),<#Pos#>(3,3))
            C12P2 : tuple(tuple(int,int,int,int),tuple(int,int)) = ((<#Type#>TANK,<#Purity#>105,<#PackA#>10,<#PackB#>10),<#Pos#>(5,3))
            C12P3 : tuple(tuple(int,int,int,int),tuple(int,int)) = ((<#Type#>DRUID,<#Purity#>103,<#PackA#>10,<#PackB#>10),<#Pos#>(4,5))
            C12P4 : tuple(tuple(int,int,int,int),tuple(int,int)) = ((<#Type#>HUNTER,<#Purity#>123,<#PackA#>10,<#PackB#>10),<#Pos#>(5,5))
        }

    AllChallenge_Pets_13 : []tuple(tuple(int,int,int,int),tuple(int,int)) = 
        array{
            C13P0 : tuple(tuple(int,int,int,int),tuple(int,int)) = ((<#Type#>ADC,<#Purity#>156,<#PackA#>11,<#PackB#>11),<#Pos#>(5,1))
            C13P1 : tuple(tuple(int,int,int,int),tuple(int,int)) = ((<#Type#>ADC,<#Purity#>123,<#PackA#>11,<#PackB#>11),<#Pos#>(5,2))
            C13P2 : tuple(tuple(int,int,int,int),tuple(int,int)) = ((<#Type#>ADC,<#Purity#>153,<#PackA#>11,<#PackB#>11),<#Pos#>(5,3))
            C13P3 : tuple(tuple(int,int,int,int),tuple(int,int)) = ((<#Type#>ADC,<#Purity#>175,<#PackA#>11,<#PackB#>11),<#Pos#>(5,4))
            C13P4 : tuple(tuple(int,int,int,int),tuple(int,int)) = ((<#Type#>ADC,<#Purity#>132,<#PackA#>11,<#PackB#>11),<#Pos#>(5,5))
        }

    AllChallenge_Pets_14 : []tuple(tuple(int,int,int,int),tuple(int,int)) = 
        array{
            C14P0 : tuple(tuple(int,int,int,int),tuple(int,int)) = ((<#Type#>TANK,<#Purity#>179,<#PackA#>13,<#PackB#>13),<#Pos#>(5,1))
            C14P1 : tuple(tuple(int,int,int,int),tuple(int,int)) = ((<#Type#>TANK,<#Purity#>201,<#PackA#>13,<#PackB#>13),<#Pos#>(5,2))
            C14P2 : tuple(tuple(int,int,int,int),tuple(int,int)) = ((<#Type#>TANK,<#Purity#>167,<#PackA#>13,<#PackB#>13),<#Pos#>(5,3))
            C14P3 : tuple(tuple(int,int,int,int),tuple(int,int)) = ((<#Type#>TANK,<#Purity#>156,<#PackA#>13,<#PackB#>13),<#Pos#>(5,4))
            C14P4 : tuple(tuple(int,int,int,int),tuple(int,int)) = ((<#Type#>TANK,<#Purity#>221,<#PackA#>13,<#PackB#>13),<#Pos#>(5,5))
        }

    AllChallenge_Pets_15 : []tuple(tuple(int,int,int,int),tuple(int,int)) = 
        array{
            C15P0 : tuple(tuple(int,int,int,int),tuple(int,int)) = ((<#Type#>FIGHTER,<#Purity#>231,<#PackA#>14,<#PackB#>14),<#Pos#>(5,1))
            C15P1 : tuple(tuple(int,int,int,int),tuple(int,int)) = ((<#Type#>FIGHTER,<#Purity#>231,<#PackA#>14,<#PackB#>14),<#Pos#>(5,2))
            C15P2 : tuple(tuple(int,int,int,int),tuple(int,int)) = ((<#Type#>FIGHTER,<#Purity#>212,<#PackA#>14,<#PackB#>14),<#Pos#>(5,3))
            C15P3 : tuple(tuple(int,int,int,int),tuple(int,int)) = ((<#Type#>FIGHTER,<#Purity#>231,<#PackA#>14,<#PackB#>14),<#Pos#>(5,4))
            C15P4 : tuple(tuple(int,int,int,int),tuple(int,int)) = ((<#Type#>FIGHTER,<#Purity#>261,<#PackA#>14,<#PackB#>14),<#Pos#>(5,5))
        }

    AllChallenge_Pets_16 : []tuple(tuple(int,int,int,int),tuple(int,int)) = 
        array{
            C16P0 : tuple(tuple(int,int,int,int),tuple(int,int)) = ((<#Type#>HUNTER,<#Purity#>212,<#PackA#>15,<#PackB#>15),<#Pos#>(3,4))
            C16P1 : tuple(tuple(int,int,int,int),tuple(int,int)) = ((<#Type#>HUNTER,<#Purity#>250,<#PackA#>15,<#PackB#>15),<#Pos#>(3,3))
            C16P2 : tuple(tuple(int,int,int,int),tuple(int,int)) = ((<#Type#>HUNTER,<#Purity#>244,<#PackA#>15,<#PackB#>15),<#Pos#>(5,3))
            C16P3 : tuple(tuple(int,int,int,int),tuple(int,int)) = ((<#Type#>HUNTER,<#Purity#>241,<#PackA#>15,<#PackB#>15),<#Pos#>(4,5))
            C16P4 : tuple(tuple(int,int,int,int),tuple(int,int)) = ((<#Type#>HUNTER,<#Purity#>254,<#PackA#>15,<#PackB#>15),<#Pos#>(5,5))
        }


    AllChallenge_Pets_17 : []tuple(tuple(int,int,int,int),tuple(int,int)) = 
        array{
            C17P0 : tuple(tuple(int,int,int,int),tuple(int,int)) = ((<#Type#>TANK,<#Purity#>276,<#PackA#>16,<#PackB#>16),<#Pos#>(3,4))
            C17P1 : tuple(tuple(int,int,int,int),tuple(int,int)) = ((<#Type#>FIGHTER,<#Purity#>254,<#PackA#>16,<#PackB#>16),<#Pos#>(3,5))
            C17P2 : tuple(tuple(int,int,int,int),tuple(int,int)) = ((<#Type#>HUNTER,<#Purity#>238,<#PackA#>16,<#PackB#>16),<#Pos#>(5,3))
            C17P3 : tuple(tuple(int,int,int,int),tuple(int,int)) = ((<#Type#>ADC,<#Purity#>267,<#PackA#>16,<#PackB#>16),<#Pos#>(4,5))
            C17P4 : tuple(tuple(int,int,int,int),tuple(int,int)) = ((<#Type#>DRUID,<#Purity#>278,<#PackA#>16,<#PackB#>16),<#Pos#>(5,5))
        }

    AllChallenge_Pets_18 : []tuple(tuple(int,int,int,int),tuple(int,int)) = 
        array{
            C18P0 : tuple(tuple(int,int,int,int),tuple(int,int)) = ((<#Type#>FIGHTER,<#Purity#>286,<#PackA#>18,<#PackB#>18),<#Pos#>(3,4))
            C18P1 : tuple(tuple(int,int,int,int),tuple(int,int)) = ((<#Type#>TANK,<#Purity#>275,<#PackA#>18,<#PackB#>18),<#Pos#>(3,3))
            C18P2 : tuple(tuple(int,int,int,int),tuple(int,int)) = ((<#Type#>DRUID,<#Purity#>287,<#PackA#>18,<#PackB#>18),<#Pos#>(5,3))
            C18P3 : tuple(tuple(int,int,int,int),tuple(int,int)) = ((<#Type#>DRUID,<#Purity#>291,<#PackA#>18,<#PackB#>18),<#Pos#>(4,5))
            C18P4 : tuple(tuple(int,int,int,int),tuple(int,int)) = ((<#Type#>DRUID,<#Purity#>270,<#PackA#>18,<#PackB#>18),<#Pos#>(5,5))
        }

    AllChallenge_Pets_19 : []tuple(tuple(int,int,int,int),tuple(int,int)) = 
        array{
            C19P0 : tuple(tuple(int,int,int,int),tuple(int,int)) = ((<#Type#>TANK,<#Purity#>291,<#PackA#>18,<#PackB#>18),<#Pos#>(3,4))
            C19P1 : tuple(tuple(int,int,int,int),tuple(int,int)) = ((<#Type#>TANK,<#Purity#>293,<#PackA#>18,<#PackB#>18),<#Pos#>(3,3))
            C19P2 : tuple(tuple(int,int,int,int),tuple(int,int)) = ((<#Type#>TANK,<#Purity#>294,<#PackA#>18,<#PackB#>18),<#Pos#>(5,3))
            C19P3 : tuple(tuple(int,int,int,int),tuple(int,int)) = ((<#Type#>TANK,<#Purity#>295,<#PackA#>18,<#PackB#>18),<#Pos#>(4,5))
            C19P4 : tuple(tuple(int,int,int,int),tuple(int,int)) = ((<#Type#>TANK,<#Purity#>291,<#PackA#>18,<#PackB#>18),<#Pos#>(5,5))
        }

    AllChallenge_Pets_20 : []tuple(tuple(int,int,int,int),tuple(int,int)) = 
        array{
            C20P0 : tuple(tuple(int,int,int,int),tuple(int,int)) = ((<#Type#>DRUID,<#Purity#>298,<#PackA#>20,<#PackB#>20),<#Pos#>(5,1))
            C20P1 : tuple(tuple(int,int,int,int),tuple(int,int)) = ((<#Type#>DRUID,<#Purity#>298,<#PackA#>20,<#PackB#>20),<#Pos#>(5,2))
            C20P2 : tuple(tuple(int,int,int,int),tuple(int,int)) = ((<#Type#>DRUID,<#Purity#>298,<#PackA#>20,<#PackB#>20),<#Pos#>(5,3))
            C20P3 : tuple(tuple(int,int,int,int),tuple(int,int)) = ((<#Type#>DRUID,<#Purity#>298,<#PackA#>20,<#PackB#>20),<#Pos#>(5,4))
            C20P4 : tuple(tuple(int,int,int,int),tuple(int,int)) = ((<#Type#>DRUID,<#Purity#>298,<#PackA#>20,<#PackB#>20),<#Pos#>(5,5))
        }

    AllChallenge_Pets_21 : []tuple(tuple(int,int,int,int),tuple(int,int)) = 
        array{
            C21P0 : tuple(tuple(int,int,int,int),tuple(int,int)) = ((<#Type#>ADC,<#Purity#>298,<#PackA#>20,<#PackB#>20),<#Pos#>(3,4))
            C21P1 : tuple(tuple(int,int,int,int),tuple(int,int)) = ((<#Type#>TANK,<#Purity#>298,<#PackA#>20,<#PackB#>20),<#Pos#>(3,3))
            C21P2 : tuple(tuple(int,int,int,int),tuple(int,int)) = ((<#Type#>TANK,<#Purity#>298,<#PackA#>20,<#PackB#>20),<#Pos#>(5,3))
            C21P3 : tuple(tuple(int,int,int,int),tuple(int,int)) = ((<#Type#>DRUID,<#Purity#>298,<#PackA#>20,<#PackB#>20),<#Pos#>(4,5))
            C21P4 : tuple(tuple(int,int,int,int),tuple(int,int)) = ((<#Type#>HUNTER,<#Purity#>298,<#PackA#>20,<#PackB#>20),<#Pos#>(5,5))
        }
    #>
    AllChallenge_Pets_22 : []tuple(tuple(int,int,int,int),tuple(int,int)) = 
        array{
            C22P0 : tuple(tuple(int,int,int,int),tuple(int,int)) = ((<#Type#>ADC,<#Purity#>299,<#PackA#>20,<#PackB#>20),<#Pos#>(5,1))
            C22P1 : tuple(tuple(int,int,int,int),tuple(int,int)) = ((<#Type#>ADC,<#Purity#>299,<#PackA#>20,<#PackB#>20),<#Pos#>(5,2))
            C22P2 : tuple(tuple(int,int,int,int),tuple(int,int)) = ((<#Type#>ADC,<#Purity#>299,<#PackA#>20,<#PackB#>20),<#Pos#>(5,3))
            C22P3 : tuple(tuple(int,int,int,int),tuple(int,int)) = ((<#Type#>ADC,<#Purity#>299,<#PackA#>20,<#PackB#>20),<#Pos#>(5,4))
            C22P4 : tuple(tuple(int,int,int,int),tuple(int,int)) = ((<#Type#>ADC,<#Purity#>299,<#PackA#>20,<#PackB#>20),<#Pos#>(5,5))
        }

    AllChallenge_Pets_23 : []tuple(tuple(int,int,int,int),tuple(int,int)) = 
        array{
            C23P0 : tuple(tuple(int,int,int,int),tuple(int,int)) = ((<#Type#>HUNTER,<#Purity#>299,<#PackA#>20,<#PackB#>20),<#Pos#>(5,1))
            C23P1 : tuple(tuple(int,int,int,int),tuple(int,int)) = ((<#Type#>HUNTER,<#Purity#>299,<#PackA#>20,<#PackB#>20),<#Pos#>(5,2))
            C23P2 : tuple(tuple(int,int,int,int),tuple(int,int)) = ((<#Type#>HUNTER,<#Purity#>299,<#PackA#>20,<#PackB#>20),<#Pos#>(5,3))
            C23P3 : tuple(tuple(int,int,int,int),tuple(int,int)) = ((<#Type#>HUNTER,<#Purity#>299,<#PackA#>20,<#PackB#>20),<#Pos#>(5,4))
            C23P4 : tuple(tuple(int,int,int,int),tuple(int,int)) = ((<#Type#>HUNTER,<#Purity#>299,<#PackA#>20,<#PackB#>20),<#Pos#>(5,5))
        }

    AllChallenge_Pets_24 : []tuple(tuple(int,int,int,int),tuple(int,int)) = 
        array{
            C24P0 : tuple(tuple(int,int,int,int),tuple(int,int)) = ((<#Type#>FIGHTER,<#Purity#>299,<#PackA#>20,<#PackB#>20),<#Pos#>(5,1))
            C24P1 : tuple(tuple(int,int,int,int),tuple(int,int)) = ((<#Type#>FIGHTER,<#Purity#>299,<#PackA#>20,<#PackB#>20),<#Pos#>(5,2))
            C24P2 : tuple(tuple(int,int,int,int),tuple(int,int)) = ((<#Type#>FIGHTER,<#Purity#>299,<#PackA#>20,<#PackB#>20),<#Pos#>(5,3))
            C24P3 : tuple(tuple(int,int,int,int),tuple(int,int)) = ((<#Type#>FIGHTER,<#Purity#>299,<#PackA#>20,<#PackB#>20),<#Pos#>(5,4))
            C24P4 : tuple(tuple(int,int,int,int),tuple(int,int)) = ((<#Type#>FIGHTER,<#Purity#>299,<#PackA#>20,<#PackB#>20),<#Pos#>(5,5))
        }

    AllChallenge_Pets_25 : []tuple(tuple(int,int,int,int),tuple(int,int)) = 
        array{
            C25P0 : tuple(tuple(int,int,int,int),tuple(int,int)) = ((<#Type#>DRUID,<#Purity#>299,<#PackA#>20,<#PackB#>20),<#Pos#>(5,1))
            C25P1 : tuple(tuple(int,int,int,int),tuple(int,int)) = ((<#Type#>DRUID,<#Purity#>299,<#PackA#>20,<#PackB#>20),<#Pos#>(5,2))
            C25P2 : tuple(tuple(int,int,int,int),tuple(int,int)) = ((<#Type#>DRUID,<#Purity#>299,<#PackA#>20,<#PackB#>20),<#Pos#>(5,3))
            C25P3 : tuple(tuple(int,int,int,int),tuple(int,int)) = ((<#Type#>DRUID,<#Purity#>299,<#PackA#>20,<#PackB#>20),<#Pos#>(5,4))
            C25P4 : tuple(tuple(int,int,int,int),tuple(int,int)) = ((<#Type#>DRUID,<#Purity#>299,<#PackA#>20,<#PackB#>20),<#Pos#>(5,5))
        }

    AllChallenge_Pets_26 : []tuple(tuple(int,int,int,int),tuple(int,int)) = 
        array{
            C26P0 : tuple(tuple(int,int,int,int),tuple(int,int)) = ((<#Type#>TANK,<#Purity#>299,<#PackA#>20,<#PackB#>20),<#Pos#>(5,1))
            C26P1 : tuple(tuple(int,int,int,int),tuple(int,int)) = ((<#Type#>TANK,<#Purity#>299,<#PackA#>20,<#PackB#>20),<#Pos#>(5,2))
            C26P2 : tuple(tuple(int,int,int,int),tuple(int,int)) = ((<#Type#>TANK,<#Purity#>299,<#PackA#>20,<#PackB#>20),<#Pos#>(5,3))
            C26P3 : tuple(tuple(int,int,int,int),tuple(int,int)) = ((<#Type#>TANK,<#Purity#>299,<#PackA#>20,<#PackB#>20),<#Pos#>(5,4))
            C26P4 : tuple(tuple(int,int,int,int),tuple(int,int)) = ((<#Type#>TANK,<#Purity#>299,<#PackA#>20,<#PackB#>20),<#Pos#>(5,5))
        }

    AllChallenge_Pets_27 : []tuple(tuple(int,int,int,int),tuple(int,int)) = 
        array{
            C27P0 : tuple(tuple(int,int,int,int),tuple(int,int)) = ((<#Type#>ADC,<#Purity#>299,<#PackA#>20,<#PackB#>20),<#Pos#>(3,4))
            C27P1 : tuple(tuple(int,int,int,int),tuple(int,int)) = ((<#Type#>TANK,<#Purity#>299,<#PackA#>20,<#PackB#>20),<#Pos#>(3,3))
            C27P2 : tuple(tuple(int,int,int,int),tuple(int,int)) = ((<#Type#>TANK,<#Purity#>299,<#PackA#>20,<#PackB#>20),<#Pos#>(3,3))
            C27P3 : tuple(tuple(int,int,int,int),tuple(int,int)) = ((<#Type#>DRUID,<#Purity#>299,<#PackA#>20,<#PackB#>20),<#Pos#>(4,5))
            C27P4 : tuple(tuple(int,int,int,int),tuple(int,int)) = ((<#Type#>HUNTER,<#Purity#>299,<#PackA#>20,<#PackB#>20),<#Pos#>(5,5))
        }

    AllChallenge_Pets_28 : []tuple(tuple(int,int,int,int),tuple(int,int)) = 
        array{
            C28P0 : tuple(tuple(int,int,int,int),tuple(int,int)) = ((<#Type#>ADC,<#Purity#>299,<#PackA#>20,<#PackB#>20),<#Pos#>(3,4))
            C28P1 : tuple(tuple(int,int,int,int),tuple(int,int)) = ((<#Type#>TANK,<#Purity#>299,<#PackA#>20,<#PackB#>20),<#Pos#>(3,2))
            C28P2 : tuple(tuple(int,int,int,int),tuple(int,int)) = ((<#Type#>TANK,<#Purity#>299,<#PackA#>20,<#PackB#>20),<#Pos#>(3,3))
            C28P3 : tuple(tuple(int,int,int,int),tuple(int,int)) = ((<#Type#>DRUID,<#Purity#>299,<#PackA#>20,<#PackB#>20),<#Pos#>(4,5))
            C28P4 : tuple(tuple(int,int,int,int),tuple(int,int)) = ((<#Type#>HUNTER,<#Purity#>299,<#PackA#>20,<#PackB#>20),<#Pos#>(5,1))
        }


    AllChallenge_Pets_29 : []tuple(tuple(int,int,int,int),tuple(int,int)) = 
        array{
            C29P0 : tuple(tuple(int,int,int,int),tuple(int,int)) = ((<#Type#>ADC,<#Purity#>299,<#PackA#>20,<#PackB#>20),<#Pos#>(4,3))
            C29P1 : tuple(tuple(int,int,int,int),tuple(int,int)) = ((<#Type#>TANK,<#Purity#>299,<#PackA#>20,<#PackB#>20),<#Pos#>(3,2))
            C29P2 : tuple(tuple(int,int,int,int),tuple(int,int)) = ((<#Type#>TANK,<#Purity#>299,<#PackA#>20,<#PackB#>20),<#Pos#>(3,3))
            C29P3 : tuple(tuple(int,int,int,int),tuple(int,int)) = ((<#Type#>ADC,<#Purity#>299,<#PackA#>20,<#PackB#>20),<#Pos#>(4,5))
            C29P4 : tuple(tuple(int,int,int,int),tuple(int,int)) = ((<#Type#>DRUID,<#Purity#>299,<#PackA#>20,<#PackB#>20),<#Pos#>(4,4))
        }

    AllChallenge_Pets_30 : []tuple(tuple(int,int,int,int),tuple(int,int)) = 
        array{
            C30P0 : tuple(tuple(int,int,int,int),tuple(int,int)) = ((<#Type#>HUNTER,<#Purity#>299,<#PackA#>20,<#PackB#>20),<#Pos#>(5,1))
            C30P1 : tuple(tuple(int,int,int,int),tuple(int,int)) = ((<#Type#>HUNTER,<#Purity#>299,<#PackA#>20,<#PackB#>20),<#Pos#>(5,4))
            C30P2 : tuple(tuple(int,int,int,int),tuple(int,int)) = ((<#Type#>ADC,<#Purity#>299,<#PackA#>20,<#PackB#>20),<#Pos#>(4,3))
            C30P3 : tuple(tuple(int,int,int,int),tuple(int,int)) = ((<#Type#>TANK,<#Purity#>299,<#PackA#>20,<#PackB#>20),<#Pos#>(3,2))
            C30P4 : tuple(tuple(int,int,int,int),tuple(int,int)) = ((<#Type#>TANK,<#Purity#>299,<#PackA#>20,<#PackB#>20),<#Pos#>(3,3))
        }

    AllChallenge_Pets_31 : []tuple(tuple(int,int,int,int),tuple(int,int)) = 
        array{
            C31P0 : tuple(tuple(int,int,int,int),tuple(int,int)) = ((<#Type#>DRUID,<#Purity#>299,<#PackA#>20,<#PackB#>20),<#Pos#>(4,3))
            C31P1 : tuple(tuple(int,int,int,int),tuple(int,int)) = ((<#Type#>ADC,<#Purity#>299,<#PackA#>20,<#PackB#>20),<#Pos#>(4,4))
            C31P2 : tuple(tuple(int,int,int,int),tuple(int,int)) = ((<#Type#>TANK,<#Purity#>299,<#PackA#>20,<#PackB#>20),<#Pos#>(3,1))
            C31P3 : tuple(tuple(int,int,int,int),tuple(int,int)) = ((<#Type#>TANK,<#Purity#>299,<#PackA#>20,<#PackB#>20),<#Pos#>(3,2))
            C31P4 : tuple(tuple(int,int,int,int),tuple(int,int)) = ((<#Type#>TANK,<#Purity#>299,<#PackA#>20,<#PackB#>20),<#Pos#>(3,3))
        }

    AllChallenge_Pets_32 : []tuple(tuple(int,int,int,int),tuple(int,int)) = 
        array{
            C32P0 : tuple(tuple(int,int,int,int),tuple(int,int)) = ((<#Type#>DRUID,<#Purity#>299,<#PackA#>20,<#PackB#>20),<#Pos#>(4,3))
            C32P1 : tuple(tuple(int,int,int,int),tuple(int,int)) = ((<#Type#>TANK,<#Purity#>299,<#PackA#>20,<#PackB#>20),<#Pos#>(3,5))
            C32P2 : tuple(tuple(int,int,int,int),tuple(int,int)) = ((<#Type#>TANK,<#Purity#>299,<#PackA#>20,<#PackB#>20),<#Pos#>(3,1))
            C32P3 : tuple(tuple(int,int,int,int),tuple(int,int)) = ((<#Type#>TANK,<#Purity#>299,<#PackA#>20,<#PackB#>20),<#Pos#>(3,2))
            C32P4 : tuple(tuple(int,int,int,int),tuple(int,int)) = ((<#Type#>TANK,<#Purity#>299,<#PackA#>20,<#PackB#>20),<#Pos#>(3,3))
        }

    AllChallenge_Pets_33 : []tuple(tuple(int,int,int,int),tuple(int,int)) = 
        array{
            C33P0 : tuple(tuple(int,int,int,int),tuple(int,int)) = ((<#Type#>DRUID,<#Purity#>299,<#PackA#>20,<#PackB#>20),<#Pos#>(4,3))
            C33P1 : tuple(tuple(int,int,int,int),tuple(int,int)) = ((<#Type#>FIGHTER,<#Purity#>299,<#PackA#>20,<#PackB#>20),<#Pos#>(3,1))
            C33P2 : tuple(tuple(int,int,int,int),tuple(int,int)) = ((<#Type#>FIGHTER,<#Purity#>299,<#PackA#>20,<#PackB#>20),<#Pos#>(3,2))
            C33P3 : tuple(tuple(int,int,int,int),tuple(int,int)) = ((<#Type#>FIGHTER,<#Purity#>299,<#PackA#>20,<#PackB#>20),<#Pos#>(3,3))
            C33P4 : tuple(tuple(int,int,int,int),tuple(int,int)) = ((<#Type#>FIGHTER,<#Purity#>299,<#PackA#>20,<#PackB#>20),<#Pos#>(3,4))
        }

    AllChallenge_Pets_34 : []tuple(tuple(int,int,int,int),tuple(int,int)) = 
        array{
            C34P0 : tuple(tuple(int,int,int,int),tuple(int,int)) = ((<#Type#>HUNTER,<#Purity#>299,<#PackA#>20,<#PackB#>20),<#Pos#>(5,2))
            C34P1 : tuple(tuple(int,int,int,int),tuple(int,int)) = ((<#Type#>FIGHTER,<#Purity#>299,<#PackA#>20,<#PackB#>20),<#Pos#>(3,1))
            C34P2 : tuple(tuple(int,int,int,int),tuple(int,int)) = ((<#Type#>FIGHTER,<#Purity#>299,<#PackA#>20,<#PackB#>20),<#Pos#>(3,2))
            C34P3 : tuple(tuple(int,int,int,int),tuple(int,int)) = ((<#Type#>FIGHTER,<#Purity#>299,<#PackA#>20,<#PackB#>20),<#Pos#>(3,3))
            C34P4 : tuple(tuple(int,int,int,int),tuple(int,int)) = ((<#Type#>FIGHTER,<#Purity#>299,<#PackA#>20,<#PackB#>20),<#Pos#>(3,4))
        }

    AllChallenge_Pets_35 : []tuple(tuple(int,int,int,int),tuple(int,int)) = 
        array{
            C35P0 : tuple(tuple(int,int,int,int),tuple(int,int)) = ((<#Type#>HUNTER,<#Purity#>299,<#PackA#>20,<#PackB#>20),<#Pos#>(5,1))
            C35P1 : tuple(tuple(int,int,int,int),tuple(int,int)) = ((<#Type#>HUNTER,<#Purity#>299,<#PackA#>20,<#PackB#>20),<#Pos#>(5,5))
            C35P2 : tuple(tuple(int,int,int,int),tuple(int,int)) = ((<#Type#>FIGHTER,<#Purity#>299,<#PackA#>20,<#PackB#>20),<#Pos#>(3,2))
            C35P3 : tuple(tuple(int,int,int,int),tuple(int,int)) = ((<#Type#>FIGHTER,<#Purity#>299,<#PackA#>20,<#PackB#>20),<#Pos#>(3,3))
            C35P4 : tuple(tuple(int,int,int,int),tuple(int,int)) = ((<#Type#>FIGHTER,<#Purity#>299,<#PackA#>20,<#PackB#>20),<#Pos#>(3,4))
        }
        
    
    SetChallenges() : void =
        set AllChallenges = AllChallenges + array{AllChallenge_Pets_0,AllChallenge_Pets_0_1,AllChallenge_Pets_0_2,AllChallenge_Pets_0_3,AllChallenge_Pets_0_4,
            AllChallenge_Pets_1,AllChallenge_Pets_1_1,AllChallenge_Pets_1_2,AllChallenge_Pets_1_3,AllChallenge_Pets_1_4,AllChallenge_Pets_1_5,AllChallenge_Pets_2,
            AllChallenge_Pets_3,AllChallenge_Pets_3_2,AllChallenge_Pets_3_3,AllChallenge_Pets_3_4,AllChallenge_Pets_3_5,AllChallenge_Pets_3_6,AllChallenge_Pets_3_7,
            AllChallenge_Pets_4,AllChallenge_Pets_5,AllChallenge_Pets_5_1}

        set AllChallenges = GenerateMidChallenges(AllChallenges)

        set AllChallenges = AllChallenges + array{AllChallenge_Pets_22,AllChallenge_Pets_23,AllChallenge_Pets_24,AllChallenge_Pets_25,AllChallenge_Pets_26,
            AllChallenge_Pets_27,AllChallenge_Pets_28,
            AllChallenge_Pets_29,AllChallenge_Pets_30,AllChallenge_Pets_31,AllChallenge_Pets_32,AllChallenge_Pets_33,
            AllChallenge_Pets_34,AllChallenge_Pets_35}

    Init() : void =
        SetChallenges()

    GenerateMidChallenges<private>(Challenges : [][]tuple(tuple(int,int,int,int),tuple(int,int))) : [][]tuple(tuple(int,int,int,int),tuple(int,int)) =

        var Output_Challenges : [][]tuple(tuple(int,int,int,int),tuple(int,int)) = Challenges

        Putiry_PerInteration := 5
        RequiredPurity_PerPackLevel := 10

        if(MaxInterations:=Floor[(295*1.0)/(Putiry_PerInteration*1.0)]):
            for(Interation:=4..MaxInterations):

                var PosVariation : int = 0
                var PurityVariation : int = 0
                var PackLevelVariation : int = 0

                var Purity : int = Putiry_PerInteration*Interation
                var SpecialPurity : int = Purity
                var SpecialPackLevel : int = PackLevelVariation

                if(Purity>295):
                    set Purity = 295
                if(var PackLevel : int =Floor[(Purity*1.0)/(RequiredPurity_PerPackLevel*1.0)]):
                    if(PackLevel>20):
                        set PackLevel = 20

                    if(IsOdd[Interation], IsOdd[PackLevel]):
                        set PosVariation = 0
                    else if(IsOdd[Interation]):
                        set PosVariation = 1
                    else:
                        set PosVariation = -1

                    #SPECIAL TYPES ONLY

                    var SPECIAL_PETTYPE : int = HUNTER

                    if(IsOdd[Interation], IsOdd[PackLevel], IsOdd[Purity]):
                        set SPECIAL_PETTYPE = HUNTER 
                        set SpecialPurity = SpecialPurity + 5
                    else if(IsEven[Interation], IsEven[PackLevel], IsEven[PackLevel]):
                        set SPECIAL_PETTYPE = FIGHTER 
                        set SpecialPurity = SpecialPurity + 5
                    else if(IsEven[Interation], IsOdd[PackLevel], IsEven[PackLevel]):
                        set SPECIAL_PETTYPE = DRUID 
                        set SpecialPurity = SpecialPurity + 20
                    else if(IsOdd[Interation], IsEven[PackLevel], IsOdd[PackLevel]):
                        set SPECIAL_PETTYPE = TANK 
                        set SpecialPurity = SpecialPurity + 5
                    else if(IsOdd[Interation], IsOdd[PackLevel], IsEven[PackLevel]):
                        set SPECIAL_PETTYPE = ADC 
                        set SpecialPurity = SpecialPurity + 15
                    else if(IsOdd[Interation], IsEven[PackLevel], IsEven[PackLevel]):
                        set SPECIAL_PETTYPE = ADC 
                        set SpecialPurity = SpecialPurity + 10
                        set SpecialPackLevel = SpecialPackLevel + 2
                    else:
                        set SPECIAL_PETTYPE = DRUID 
                        set SpecialPurity = SpecialPurity + 5
                        set SpecialPackLevel = SpecialPackLevel + 2

                    
                    if(SpecialPurity>295):
                        set SpecialPurity = 295
                        ChallengeTypeSpecial : []tuple(tuple(int,int,int,int),tuple(int,int)) = 
                            array{
                                ((<#Type#>SPECIAL_PETTYPE,<#Purity#>Purity+PurityVariation,<#PackA#>PackLevel+SpecialPackLevel,<#PackB#>PackLevel+SpecialPackLevel),<#Pos#>(4+PosVariation,2))
                                ((<#Type#>SPECIAL_PETTYPE,<#Purity#>Purity+PurityVariation,<#PackA#>PackLevel+SpecialPackLevel,<#PackB#>PackLevel+SpecialPackLevel),<#Pos#>(4+PosVariation,3))
                                ((<#Type#>SPECIAL_PETTYPE,<#Purity#>Purity+PurityVariation,<#PackA#>PackLevel+SpecialPackLevel,<#PackB#>PackLevel+SpecialPackLevel),<#Pos#>(4+PosVariation,4))
                                ((<#Type#>SPECIAL_PETTYPE,<#Purity#>Purity+PurityVariation,<#PackA#>PackLevel+SpecialPackLevel,<#PackB#>PackLevel+SpecialPackLevel),<#Pos#>(4+PosVariation,5))
                                ((<#Type#>SPECIAL_PETTYPE,<#Purity#>Purity+PurityVariation,<#PackA#>PackLevel+SpecialPackLevel,<#PackB#>PackLevel+SpecialPackLevel),<#Pos#>(4+PosVariation,6))
                            }   
                        set Output_Challenges = Output_Challenges + array{ChallengeTypeSpecial}
                    else if(SpecialPurity>250):
                        ChallengeTypeSpecial : []tuple(tuple(int,int,int,int),tuple(int,int)) = 
                            array{
                                ((<#Type#>SPECIAL_PETTYPE,<#Purity#>Purity+PurityVariation,<#PackA#>PackLevel+SpecialPackLevel,<#PackB#>PackLevel+SpecialPackLevel),<#Pos#>(4+PosVariation,2))
                                ((<#Type#>SPECIAL_PETTYPE,<#Purity#>Purity+PurityVariation,<#PackA#>PackLevel+SpecialPackLevel,<#PackB#>PackLevel+SpecialPackLevel),<#Pos#>(4+PosVariation,3))
                                ((<#Type#>SPECIAL_PETTYPE,<#Purity#>Purity+PurityVariation,<#PackA#>PackLevel+SpecialPackLevel,<#PackB#>PackLevel+SpecialPackLevel),<#Pos#>(4+PosVariation,4))
                                ((<#Type#>SPECIAL_PETTYPE,<#Purity#>Purity+PurityVariation,<#PackA#>PackLevel+SpecialPackLevel,<#PackB#>PackLevel+SpecialPackLevel),<#Pos#>(4+PosVariation,5))
                            }
                    else if(SpecialPurity>200):
                        ChallengeTypeSpecial : []tuple(tuple(int,int,int,int),tuple(int,int)) = 
                            array{
                                ((<#Type#>SPECIAL_PETTYPE,<#Purity#>Purity+PurityVariation,<#PackA#>PackLevel+SpecialPackLevel,<#PackB#>PackLevel+SpecialPackLevel),<#Pos#>(4+PosVariation,2))
                                ((<#Type#>SPECIAL_PETTYPE,<#Purity#>Purity+PurityVariation,<#PackA#>PackLevel+SpecialPackLevel,<#PackB#>PackLevel+SpecialPackLevel),<#Pos#>(4+PosVariation,3))
                                ((<#Type#>SPECIAL_PETTYPE,<#Purity#>Purity+PurityVariation,<#PackA#>PackLevel+SpecialPackLevel,<#PackB#>PackLevel+SpecialPackLevel),<#Pos#>(4+PosVariation,4))
                            }
                        set Output_Challenges = Output_Challenges + array{ChallengeTypeSpecial}
                    #NORMALS
                    ChallengeType0 : []tuple(tuple(int,int,int,int),tuple(int,int)) = 
                        array{
                            ((<#Type#>TANK,<#Purity#>Purity+PurityVariation,<#PackA#>PackLevel+PackLevelVariation,<#PackB#>PackLevel+PackLevelVariation),<#Pos#>(3,3+PosVariation))
                            ((<#Type#>FIGHTER,<#Purity#>Purity+PurityVariation,<#PackA#>PackLevel+PackLevelVariation,<#PackB#>PackLevel+PackLevelVariation),<#Pos#>(3,5))
                            ((<#Type#>HUNTER,<#Purity#>Purity+PurityVariation,<#PackA#>PackLevel+PackLevelVariation,<#PackB#>PackLevel+PackLevelVariation),<#Pos#>(5,1+PosVariation))
                            ((<#Type#>ADC,<#Purity#>Purity+PurityVariation,<#PackA#>PackLevel+PackLevelVariation,<#PackB#>PackLevel+PackLevelVariation),<#Pos#>(5,3))
                            ((<#Type#>DRUID,<#Purity#>Purity+PurityVariation,<#PackA#>PackLevel+PackLevelVariation,<#PackB#>PackLevel+PackLevelVariation),<#Pos#>(5,5+PosVariation))
                        }
                    set Output_Challenges = Output_Challenges + array{ChallengeType0}
                    ChallengeType1 : []tuple(tuple(int,int,int,int),tuple(int,int)) = 
                        array{
                            ((<#Type#>TANK,<#Purity#>Purity+PurityVariation,<#PackA#>PackLevel+PackLevelVariation,<#PackB#>PackLevel+PackLevelVariation),<#Pos#>(3,2))
                            ((<#Type#>TANK,<#Purity#>Purity+PurityVariation,<#PackA#>PackLevel+PackLevelVariation,<#PackB#>PackLevel+PackLevelVariation),<#Pos#>(3,4+PosVariation))
                            ((<#Type#>ADC,<#Purity#>Purity+PurityVariation,<#PackA#>PackLevel+PackLevelVariation,<#PackB#>PackLevel+PackLevelVariation),<#Pos#>(4,3))
                            ((<#Type#>DRUID,<#Purity#>Purity+PurityVariation,<#PackA#>PackLevel+PackLevelVariation,<#PackB#>PackLevel+PackLevelVariation),<#Pos#>(4,5+PosVariation))
                            ((<#Type#>HUNTER,<#Purity#>Purity+PurityVariation,<#PackA#>PackLevel+PackLevelVariation,<#PackB#>PackLevel+PackLevelVariation),<#Pos#>(5,2+PosVariation))
                        }
                    set Output_Challenges = Output_Challenges + array{ChallengeType1}
                    ChallengeType2 : []tuple(tuple(int,int,int,int),tuple(int,int)) = 
                        array{
                            ((<#Type#>FIGHTER,<#Purity#>Purity+PurityVariation,<#PackA#>PackLevel+PackLevelVariation,<#PackB#>PackLevel+PackLevelVariation),<#Pos#>(3,2))
                            ((<#Type#>FIGHTER,<#Purity#>Purity+PurityVariation,<#PackA#>PackLevel+PackLevelVariation,<#PackB#>PackLevel+PackLevelVariation),<#Pos#>(3,3))
                            ((<#Type#>FIGHTER,<#Purity#>Purity+PurityVariation,<#PackA#>PackLevel+PackLevelVariation,<#PackB#>PackLevel+PackLevelVariation),<#Pos#>(3,4))
                            ((<#Type#>DRUID,<#Purity#>Purity+PurityVariation,<#PackA#>PackLevel+PackLevelVariation,<#PackB#>PackLevel+PackLevelVariation),<#Pos#>(4,4+PosVariation))
                            ((<#Type#>SPECIAL_PETTYPE,<#Purity#>Purity+PurityVariation,<#PackA#>PackLevel+PackLevelVariation,<#PackB#>PackLevel+PackLevelVariation),<#Pos#>(4,6))
                        }
                    set Output_Challenges = Output_Challenges + array{ChallengeType2}
        return Output_Challenges

    GetChallengeName(ChallengeNumber : int) : tuple(string,logic) =
        var IsValid : logic = false
        var String : string = "-"
        if(ChallengeNumber<AllChallenges.Length, ChallengeNumber >= 0):
            set IsValid = true
            set String = "Challenge {ChallengeNumber+1}"
        return (String,IsValid)

    GetChallengeGroup(HexagamePets : []hexagame_pet, HexaMiniGameManager : hexa_minigame_manager, GameManager : game_manager, ChallengeNumber : int) : []hexagame_pet =
        if(AllChallegePetStats:=AllChallenges[ChallengeNumber]):
            for(Index:=0..HexagamePets.Length-1, PetStats := AllChallegePetStats[Index], AIHexaPet := HexagamePets[Index]):

                PetType := PetStats(0)(0)
                PetRawPurity := PetStats(0)(1)
                PackA := PetStats(0)(2)
                PackB := PetStats(0)(3)
                Position := PetStats(1)

                TuplePurity := ConvertRawPurity(PetRawPurity)

                PetTier:=TuplePurity(0)
                PetPurity:=TuplePurity(1)

                if(Position(0)<0 or Position(0)>5):
                    PrintFatalError("On GetChallengeGroup Column invalid!")
                if(Position(1)<0 or Position(1)>6):
                    PrintFatalError("On GetChallengeGroup Row invalid!")

                MaxHealth := CalculePetStat(STAT_TYPE_HEALTH,PackA,PetType,PetRawPurity)(0)
                AttackDamage := CalculePetStat(STAT_TYPE_DAMAGE,PackA,PetType,PetRawPurity)(0)
                Speed := CalculePetStat(STAT_TYPE_SPEED,PackA,PetType,PetRawPurity)(0)
                
                ActionSpeed := CalculePetStat(STAT_TYPE_ATTACKSPEED,PackB,PetType,PetRawPurity)(0)
                HealingPerk := CalculePetStat(STAT_TYPE_HEALING,PackB,PetType,PetRawPurity)(0)

                AttackRange:=ConvertRange(true, CalculePetStat(STAT_TYPE_RANGE,PackB,PetType,PetRawPurity)(0))
                DefendRange:=ConvertRange(true, CalculePetStat(STAT_TYPE_RANGE,PackB,PetType,PetRawPurity)(0))

                if:
                    InAttackRange := Floor[AttackRange]
                    InDefendRange := Floor[DefendRange]
                then:
                    AIHexaPet.SetNewPetStats(HexaMiniGameManager, GameManager.PropsManager.GetPetProp(PetType), PackA,PackB,MaxHealth,
                        AttackDamage, InAttackRange, InDefendRange,
                        PetType, false, PetTier, PetPurity,
                        Speed, ActionSpeed, HealingPerk)
                    AIHexaPet.SetIsBeingUsed(true)
                    AIHexaPet.SetInitialPosition(Position)
        return HexagamePets
                    

    

    

