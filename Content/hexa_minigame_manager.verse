using { /Fortnite.com/Devices }
using { /Verse.org/Simulation }
using { /UnrealEngine.com/Temporary/Diagnostics }
using { /UnrealEngine.com/Temporary/SpatialMath }
using { /UnrealEngine.com/Temporary/UI }
using { /Verse.org/Random }
using { /Fortnite.com/UI }
using { /Verse.org/Colors/NamedColors }
using { /Verse.org/Assets }


# See https://dev.epicgames.com/documentation/en-us/uefn/create-your-own-device-in-verse for how to create a verse device.

# A Verse-authored creative device that can be placed in a level
hexa_minigame_manager := class(creative_device):

    var GameIsRunning : logic = false
    var IsUsingHexaGame : logic = false

    var MaybeBakeryMachinesManager : ?bakery_machines_manager = false

    var MaybeAgent : ?agent = false
    var MaybeNavigationCanvas : ?canvas = false
    var MaybeChallengeCanvas : ?canvas = false

    var UIIsChallengeCanvas : logic = false
    var UIIsNavigationCanvas : logic = false

    PlayersDataManager : player_data_manager = player_data_manager{}

    @editable
    CameraFixedPoint_Hexa : gameplay_camera_fixed_point_device = gameplay_camera_fixed_point_device{}
    @editable
    CameraFixedPoint_HexaEnemies : gameplay_camera_fixed_point_device = gameplay_camera_fixed_point_device{}

    @editable
    HexHeight : float = 32.0

    @editable
    HexaNumColumns : int = 7
    @editable
    HexaNumRows : int = 6
    @editable
    HexaMaxRowDesign : int = 2


    @editable
    MeleeMaxRange : int = 384
    @editable
    ADCMaxRange : int = 640


    @editable
    AllHex : []creative_prop =array{}

    var SelectedPetProp : creative_prop = creative_prop{}
    var MaybeSelectedPet : ?hexagame_pet = false

    var AllHexBoxes : [][]?hexagame_pet = array{}
    var AllHexPositions : [][]creative_prop = array{}
    var AgentFocus : tuple(int,int) = (0,0)

    var MaybeHexMiniGameUI : ?hexa_minigame_ui = false
    var MaybeHexMiniGameSelectorUI : ?hexa_minigame_selector_ui = false

    var AllHexaAIPets : []hexagame_pet = array{}
    var AllHexaPlayerPets : []hexagame_pet = array{}

    #UI Navigation
    ButtonNavigationLeft : button_regular = button_regular{}
    ButtonNavigationRight : button_regular = button_regular{}
    ButtonNavigationOK: button_regular = button_regular{}
    ButtonNavigationTop: button_regular = button_regular{}
    ButtonNavigationBottom: button_regular = button_regular{}
    ButtonNavigationCancel: button_loud = button_loud{}
    NumPetsPlaced : text_block = text_block{DefaultTextColor:=White,DefaultShadowOffset:=ShadowOffset, DefaultShadowOpacity := 1.0}
    #ButtonNavigationFight: button_regular = button_regular{}

    #UI Challenges
    HexaChallengesDatabase : hexa_challenges_database = hexa_challenges_database{}

    var ChallengeCurrentSelection : int = 0

    SlotChallenge_Image1 : texture_block = texture_block{DefaultImage:=Images.Cross_Icon, DefaultDesiredSize := vector2{X:=75.0,Y:=75.0}}
    SlotChallenge_Image2 : texture_block = texture_block{DefaultImage:=Images.Cross_Icon, DefaultDesiredSize := vector2{X:=75.0,Y:=75.0}}
    SlotChallenge_Image3 : texture_block = texture_block{DefaultImage:=Images.Cross_Icon, DefaultDesiredSize := vector2{X:=75.0,Y:=75.0}}
    SlotChallenge_Image4 : texture_block = texture_block{DefaultImage:=Images.Cross_Icon, DefaultDesiredSize := vector2{X:=75.0,Y:=75.0}}
    SlotChallenge_Image5 : texture_block = texture_block{DefaultImage:=Images.Cross_Icon, DefaultDesiredSize := vector2{X:=75.0,Y:=75.0}}

    SlotChallenge_ExtraRewardsInfo1 : text_block = text_block{DefaultTextColor:=White, DefaultShadowOffset:=ShadowOffset, DefaultShadowOpacity := 1.0}
    SlotChallenge_ExtraRewardsInfo2 : text_block = text_block{DefaultTextColor:=White, DefaultShadowOffset:=ShadowOffset, DefaultShadowOpacity := 1.0}
    SlotChallenge_ExtraRewardsInfo3 : text_block = text_block{DefaultTextColor:=White, DefaultShadowOffset:=ShadowOffset, DefaultShadowOpacity := 1.0}
    SlotChallenge_ExtraRewardsInfo4 : text_block = text_block{DefaultTextColor:=White, DefaultShadowOffset:=ShadowOffset, DefaultShadowOpacity := 1.0}
    SlotChallenge_ExtraRewardsInfo5 : text_block = text_block{DefaultTextColor:=White, DefaultShadowOffset:=ShadowOffset, DefaultShadowOpacity := 1.0}

    SlotChallenge_Description1 : text_block = text_block{DefaultTextColor:=White, DefaultShadowOffset:=ShadowOffset, DefaultShadowOpacity := 1.0}
    SlotChallenge_Description2 : text_block = text_block{DefaultTextColor:=White, DefaultShadowOffset:=ShadowOffset, DefaultShadowOpacity := 1.0}
    SlotChallenge_Description3 : text_block = text_block{DefaultTextColor:=White, DefaultShadowOffset:=ShadowOffset, DefaultShadowOpacity := 1.0}
    SlotChallenge_Description4 : text_block = text_block{DefaultTextColor:=White, DefaultShadowOffset:=ShadowOffset, DefaultShadowOpacity := 1.0}
    SlotChallenge_Description5 : text_block = text_block{DefaultTextColor:=White, DefaultShadowOffset:=ShadowOffset, DefaultShadowOpacity := 1.0}

    SlotChallenge_IsCompleted1 : text_block = text_block{DefaultTextColor:=White, DefaultShadowOffset:=ShadowOffset, DefaultShadowOpacity := 1.0}
    SlotChallenge_IsCompleted2 : text_block = text_block{DefaultTextColor:=White, DefaultShadowOffset:=ShadowOffset, DefaultShadowOpacity := 1.0}
    SlotChallenge_IsCompleted3 : text_block = text_block{DefaultTextColor:=White, DefaultShadowOffset:=ShadowOffset, DefaultShadowOpacity := 1.0}
    SlotChallenge_IsCompleted4 : text_block = text_block{DefaultTextColor:=White, DefaultShadowOffset:=ShadowOffset, DefaultShadowOpacity := 1.0}
    SlotChallenge_IsCompleted5 : text_block = text_block{DefaultTextColor:=White, DefaultShadowOffset:=ShadowOffset, DefaultShadowOpacity := 1.0}

    var AllSlotChallenge_Images : []texture_block = array{}
    var AllSlotChallenge_Description : []text_block = array{}
    var AllSlotChallenge_IsCompleted : []text_block = array{}

    NumChallengesCompleted : text_block = text_block{DefaultTextColor:=White,DefaultShadowOffset:=ShadowOffset, DefaultShadowOpacity := 1.0}
    CoalReward_Text : text_block = text_block{DefaultTextColor:=CoalTextColor, DefaultShadowOffset:=ShadowOffset, DefaultShadowOpacity := DefaultTextShadowOpacity}
    MutantMaterialReward_Text : text_block = text_block{DefaultTextColor:=MutantMaterialTextColor, DefaultShadowOffset:=ShadowOffset, DefaultShadowOpacity := DefaultTextShadowOpacity}
 

    ButtonChallengesLeft : button_quiet = button_quiet{}
    ButtonChallenges5Left : button_quiet = button_quiet{}
    ButtonChallengesRight : button_quiet = button_quiet{}
    ButtonChallenges5Right : button_quiet = button_quiet{}
    ButtonChallengesOK: button_regular = button_regular{}
    ButtonChallengesClose: button_loud = button_loud{}

    #Slot 1 AI
    EmptyImage_AI_Slot1 : texture_block = texture_block{DefaultImage:=Images.EmptyTexture, DefaultDesiredSize := vector2{X:=SizeSlotX_PetHexaUI,Y:=SizeSlotY_PetHexaUI}}
    HealthLine_AI_Slot1 : texture_block = texture_block{DefaultTint:=Red, DefaultImage:=Images.Img_TintedBackgroundHighOpacity, DefaultDesiredSize := vector2{X:=0.0,Y:=0.0}}
    PetImage_AI_Slot1 : texture_block = texture_block{DefaultImage:=Images.Img_LineBackground, DefaultDesiredSize := vector2{X:=SizeImageX_PetHexaUI,Y:=SizeImageY_PetHexaUI}}
    Class_AI_Slot1 : text_block = text_block{DefaultTextColor:=White,DefaultShadowOffset:=ShadowOffset, DefaultShadowOpacity := 1.0}
    PackA_AI_Box1 : text_block = text_block{DefaultTextColor := White, DefaultShadowOffset:=ShadowOffset, DefaultShadowOpacity := 1.0}
    PackB_AI_Box1 : text_block = text_block{DefaultTextColor := White, DefaultShadowOffset:=ShadowOffset, DefaultShadowOpacity := 1.0}
    Purity_AI_Slot1 : text_block = text_block{DefaultTextColor:=White,DefaultShadowOffset:=ShadowOffset, DefaultShadowOpacity := 1.0}
    Health_AI_Slot1 : text_block = text_block{DefaultTextColor:=White,DefaultShadowOffset:=ShadowOffset, DefaultShadowOpacity := 1.0}

    #Slot 2 AI
    EmptyImage_AI_Slot2 : texture_block = texture_block{DefaultImage:=Images.EmptyTexture, DefaultDesiredSize := vector2{X:=SizeSlotX_PetHexaUI,Y:=SizeSlotY_PetHexaUI}}
    HealthLine_AI_Slot2 : texture_block = texture_block{DefaultTint:=Red, DefaultImage:=Images.Img_TintedBackgroundHighOpacity, DefaultDesiredSize := vector2{X:=0.0,Y:=0.0}}
    PetImage_AI_Slot2 : texture_block = texture_block{DefaultImage:=Images.Img_LineBackground, DefaultDesiredSize := vector2{X:=SizeImageX_PetHexaUI,Y:=SizeImageY_PetHexaUI}}
    Class_AI_Slot2 : text_block = text_block{DefaultTextColor:=White,DefaultShadowOffset:=ShadowOffset, DefaultShadowOpacity := 1.0}
    PackA_AI_Box2 : text_block = text_block{DefaultTextColor := White, DefaultShadowOffset:=ShadowOffset, DefaultShadowOpacity := 1.0}
    PackB_AI_Box2 : text_block = text_block{DefaultTextColor := White, DefaultShadowOffset:=ShadowOffset, DefaultShadowOpacity := 1.0}
    Purity_AI_Slot2 : text_block = text_block{DefaultTextColor:=White,DefaultShadowOffset:=ShadowOffset, DefaultShadowOpacity := 1.0}
    Health_AI_Slot2 : text_block = text_block{DefaultTextColor:=White,DefaultShadowOffset:=ShadowOffset, DefaultShadowOpacity := 1.0}

    #Slot 3 AI
    EmptyImage_AI_Slot3 : texture_block = texture_block{DefaultImage:=Images.EmptyTexture, DefaultDesiredSize := vector2{X:=SizeSlotX_PetHexaUI,Y:=SizeSlotY_PetHexaUI}}
    HealthLine_AI_Slot3 : texture_block = texture_block{DefaultTint:=Red, DefaultImage:=Images.Img_TintedBackgroundHighOpacity, DefaultDesiredSize := vector2{X:=0.0,Y:=0.0}}
    PetImage_AI_Slot3 : texture_block = texture_block{DefaultImage:=Images.Img_LineBackground, DefaultDesiredSize := vector2{X:=SizeImageX_PetHexaUI,Y:=SizeImageY_PetHexaUI}}
    Class_AI_Slot3 : text_block = text_block{DefaultTextColor:=White,DefaultShadowOffset:=ShadowOffset, DefaultShadowOpacity := 1.0}
    PackA_AI_Box3 : text_block = text_block{DefaultTextColor := White, DefaultShadowOffset:=ShadowOffset, DefaultShadowOpacity := 1.0}
    PackB_AI_Box3 : text_block = text_block{DefaultTextColor := White, DefaultShadowOffset:=ShadowOffset, DefaultShadowOpacity := 1.0}
    Purity_AI_Slot3 : text_block = text_block{DefaultTextColor:=White,DefaultShadowOffset:=ShadowOffset, DefaultShadowOpacity := 1.0}
    Health_AI_Slot3 : text_block = text_block{DefaultTextColor:=White,DefaultShadowOffset:=ShadowOffset, DefaultShadowOpacity := 1.0}

    #Slot 4 AI
    EmptyImage_AI_Slot4 : texture_block = texture_block{DefaultImage:=Images.EmptyTexture, DefaultDesiredSize := vector2{X:=SizeSlotX_PetHexaUI,Y:=SizeSlotY_PetHexaUI}}
    HealthLine_AI_Slot4 : texture_block = texture_block{DefaultTint:=Red, DefaultImage:=Images.Img_TintedBackgroundHighOpacity, DefaultDesiredSize := vector2{X:=0.0,Y:=0.0}}
    PetImage_AI_Slot4 : texture_block = texture_block{DefaultImage:=Images.Img_LineBackground, DefaultDesiredSize := vector2{X:=SizeImageX_PetHexaUI,Y:=SizeImageY_PetHexaUI}}
    Class_AI_Slot4 : text_block = text_block{DefaultTextColor:=White,DefaultShadowOffset:=ShadowOffset, DefaultShadowOpacity := 1.0}
    PackA_AI_Box4 : text_block = text_block{DefaultTextColor := White, DefaultShadowOffset:=ShadowOffset, DefaultShadowOpacity := 1.0}
    PackB_AI_Box4 : text_block = text_block{DefaultTextColor := White, DefaultShadowOffset:=ShadowOffset, DefaultShadowOpacity := 1.0}
    Purity_AI_Slot4 : text_block = text_block{DefaultTextColor:=White,DefaultShadowOffset:=ShadowOffset, DefaultShadowOpacity := 1.0}
    Health_AI_Slot4 : text_block = text_block{DefaultTextColor:=White,DefaultShadowOffset:=ShadowOffset, DefaultShadowOpacity := 1.0}

    #Slot 5 AI
    EmptyImage_AI_Slot5 : texture_block = texture_block{DefaultImage:=Images.EmptyTexture, DefaultDesiredSize := vector2{X:=SizeSlotX_PetHexaUI,Y:=SizeSlotY_PetHexaUI}}
    HealthLine_AI_Slot5 : texture_block = texture_block{DefaultTint:=Red, DefaultImage:=Images.Img_TintedBackgroundHighOpacity, DefaultDesiredSize := vector2{X:=0.0,Y:=0.0}}
    PetImage_AI_Slot5 : texture_block = texture_block{DefaultImage:=Images.Img_LineBackground, DefaultDesiredSize := vector2{X:=SizeImageX_PetHexaUI,Y:=SizeImageY_PetHexaUI}}
    Class_AI_Slot5 : text_block = text_block{DefaultTextColor:=White,DefaultShadowOffset:=ShadowOffset, DefaultShadowOpacity := 1.0}
    PackA_AI_Box5 : text_block = text_block{DefaultTextColor := White, DefaultShadowOffset:=ShadowOffset, DefaultShadowOpacity := 1.0}
    PackB_AI_Box5 : text_block = text_block{DefaultTextColor := White, DefaultShadowOffset:=ShadowOffset, DefaultShadowOpacity := 1.0}
    Purity_AI_Slot5 : text_block = text_block{DefaultTextColor:=White,DefaultShadowOffset:=ShadowOffset, DefaultShadowOpacity := 1.0}
    Health_AI_Slot5 : text_block = text_block{DefaultTextColor:=White,DefaultShadowOffset:=ShadowOffset, DefaultShadowOpacity := 1.0}

    var AllTextureBlock_HealthLines_AI : []texture_block = array{}
    var AllTextureBlock_PetImages_AI : []texture_block = array{}
    var AllTextBlock_Classes_AI : []text_block = array{}
    var AllTextBlock_PackA_AI : []text_block = array{}
    var AllTextBlock_PackB_AI : []text_block = array{}
    var AllTextBlock_Purity_AI : []text_block = array{}
    var AllTextBlock_Health_AI : []text_block = array{}
        
    CompletedImage : texture = Images.CheckIcon
    NotCompletedImage : texture = Images.Cross_Icon

    # Runs when the device is started in a running game
    InitUI<private>() : void =

        HexaChallengesDatabase.Init() 

        set AllTextBlock_PackA_AI = array{PackA_AI_Box1,PackA_AI_Box2,PackA_AI_Box3,PackA_AI_Box4,PackA_AI_Box5}
        set AllTextBlock_PackB_AI = array{PackB_AI_Box1,PackB_AI_Box2,PackB_AI_Box3,PackB_AI_Box4,PackB_AI_Box5}

        set AllSlotChallenge_Images = array{SlotChallenge_Image1,SlotChallenge_Image2,SlotChallenge_Image3,SlotChallenge_Image4,SlotChallenge_Image5}
        set AllSlotChallenge_Description = array{SlotChallenge_Description1,SlotChallenge_Description2,SlotChallenge_Description3,SlotChallenge_Description4,SlotChallenge_Description5}
        set AllSlotChallenge_IsCompleted = array{SlotChallenge_IsCompleted1,SlotChallenge_IsCompleted2,SlotChallenge_IsCompleted3,SlotChallenge_IsCompleted4,SlotChallenge_IsCompleted5}

        set AllTextureBlock_HealthLines_AI = array{HealthLine_AI_Slot1,HealthLine_AI_Slot2,HealthLine_AI_Slot3,HealthLine_AI_Slot4,HealthLine_AI_Slot5}
        set AllTextureBlock_PetImages_AI = array{PetImage_AI_Slot1,PetImage_AI_Slot2,PetImage_AI_Slot3,PetImage_AI_Slot4,PetImage_AI_Slot5}
        set AllTextBlock_Classes_AI = array{Class_AI_Slot1,Class_AI_Slot2,Class_AI_Slot3,Class_AI_Slot4,Class_AI_Slot5}
        set AllTextBlock_PackA_AI = array{PackA_AI_Box1,PackA_AI_Box2,PackA_AI_Box3,PackA_AI_Box4,PackA_AI_Box5}
        set AllTextBlock_PackB_AI = array{PackB_AI_Box1,PackB_AI_Box2,PackB_AI_Box3,PackB_AI_Box4,PackB_AI_Box5}
        set AllTextBlock_Purity_AI = array{Purity_AI_Slot1,Purity_AI_Slot2,Purity_AI_Slot3,Purity_AI_Slot4,Purity_AI_Slot5}
        set AllTextBlock_Health_AI = array{Health_AI_Slot1,Health_AI_Slot2,Health_AI_Slot3,Health_AI_Slot4,Health_AI_Slot5}

        HexMiniGameSelectorUI := MakeHexMiniGameSelectorUI(Self)
        HexMiniGameSelectorUI.InitUI()
        HexMiniGameSelectorUI.RefreshAll()
        set MaybeHexMiniGameSelectorUI = option{HexMiniGameSelectorUI}

        HexMiniGameUI := MakeHexaMinigameUI(Self)
        HexMiniGameUI.InitUI()
        HexMiniGameUI.RefreshAll()
        set MaybeHexMiniGameUI = option{HexMiniGameUI}

    SetBakeryManager(BakeryMachinesManager : bakery_machines_manager) : void =
        set MaybeBakeryMachinesManager = option{BakeryMachinesManager}

    Init():void=
        set AllHexaAIPets = array{}
        set AllHexaPlayerPets = array{}

        InitUI()
        InitUIButtons()

        if(BakeryMachinesManager := MaybeBakeryMachinesManager?):
            for(Index:=0..4):
                HexaPet := MakeHexaPetTable(-1,false,BakeryMachinesManager.GameManager.PropsManager.GetPetProp(DRUID),
                    BakeryMachinesManager.GameManager.VFXManager.GetMeleeAttackProp(),
                    BakeryMachinesManager.GameManager.VFXManager.GetRangeAttackProp(),
                    BakeryMachinesManager.GameManager.VFXManager.GetDefendProp(),
                    1,1,
                    1,1,DRUID,AI_PET,1,1,1.0,1.0,1.0,1.0,1.0)
                HexaPet.SetIsBeingUsed(false)
                set AllHexaAIPets = AllHexaAIPets + array{HexaPet}

        NavCanvas := MakeNavigationCanvas()
        set MaybeNavigationCanvas = option{NavCanvas}

        ChallengeCanvas := MakeChallengeCanvas()
        set MaybeChallengeCanvas = option{ChallengeCanvas}

        #Init Matrix
        if(FirstProp := AllHex[0]):
            set AllHexPositions=
                for(Row:=0..HexaNumRows-1):
                    for(Column:=0..HexaNumColumns-1):
                        FirstProp

        #Init Box Matrix
        set AllHexBoxes=
            for(Row:=0..HexaNumRows-1):
                for(Column:=0..HexaNumColumns-1):
                    false

        for(IndexRow:=0..HexaNumRows-1):
            for(IndexColumn:=0..HexaNumColumns-1, HexaProp := AllHex[(HexaNumColumns*IndexRow)+IndexColumn]):
                if(set AllHexPositions[IndexRow][IndexColumn] = HexaProp):

    CloseAllUI() : void =
        Remove_NavigationCanvasWithoutMessage()
        Remove_ChallengeCanvasWithoutMessage()
        
        if(HexMiniGameUI:=MaybeHexMiniGameUI?,HexMiniGameSelectorUI:=MaybeHexMiniGameSelectorUI?):
            HexMiniGameUI.Remove_CanvasWithoutMessage()
            HexMiniGameSelectorUI.Remove_CanvasWithoutMessage()

    ForceCloseGame() : void =
        ResetLastGame(true)
        Remove_NavigationCanvasWithoutMessage()
        Remove_ChallengeCanvasWithoutMessage()
        if(HexaMinigameManagerUI := MaybeHexMiniGameUI?):
            HexaMinigameManagerUI.Remove_CanvasWithoutMessage()
        if(Agent:=MaybeAgent?):
            if(BakeryMachinesManager := MaybeBakeryMachinesManager?):
                spawn:
                    BakeryMachinesManager.GameManager.VFXManager.StopSound(HEXAGAMEMUSIC_SOUND,0.0,option{Agent},false)
            CameraFixedPoint_HexaEnemies.RemoveFrom(Agent)
            CameraFixedPoint_Hexa.RemoveFrom(Agent)

    InitUIButtons() : void =
        #CHALLENGES
        ButtonChallengesLeft.SetText(BasicMessage("◀ "))
        ButtonChallenges5Left.SetText(BasicMessage("◀ x5"))
        ButtonChallenges5Left.OnClick().Subscribe(NavigateChallenges5Left)
        ButtonChallenges5Right.OnClick().Subscribe(NavigateChallenges5Right)
        ButtonChallengesLeft.OnClick().Subscribe(NavigateChallengesLeft)
        ButtonChallengesRight.SetText(BasicMessage("▶ "))
        ButtonChallenges5Right.SetText(BasicMessage("▶ x5"))
        ButtonChallengesRight.OnClick().Subscribe(NavigateChallengesRight)
        ButtonChallengesClose.SetText(BasicMessage("EXIT CHALLENGES SELECTOR"))
        ButtonChallengesClose.OnClick().Subscribe(Remove_ChallengeCanvasAndCamera)
        ButtonChallengesOK.SetText(BasicMessage("CHALLENGE"))
        ButtonChallengesOK.OnClick().Subscribe(Handler_ChallengeSelector_ToPetSelector)

        #NAVIGATION
        ButtonNavigationLeft.SetText(BasicMessage("◀ "))
        ButtonNavigationRight.SetText(BasicMessage("▶ "))
        ButtonNavigationTop.SetText(BasicMessage("▲"))
        ButtonNavigationBottom.SetText(BasicMessage("▼"))
        ButtonNavigationOK.SetText(BasicMessage("OK"))
        ButtonNavigationCancel.SetText(BasicMessage("CANCEL"))
        #ButtonNavigationFight.SetText(BasicMessage("FIGHT"))

        ButtonNavigationBottom.OnClick().Subscribe(handler_hexgame{MaybeHexaMinigameManager:=option{Self},Direction:=0}.HandlerDirection)
        ButtonNavigationTop.OnClick().Subscribe(handler_hexgame{MaybeHexaMinigameManager:=option{Self},Direction:=1}.HandlerDirection)
        ButtonNavigationLeft.OnClick().Subscribe(handler_hexgame{MaybeHexaMinigameManager:=option{Self},Direction:=2}.HandlerDirection)
        ButtonNavigationRight.OnClick().Subscribe(handler_hexgame{MaybeHexaMinigameManager:=option{Self},Direction:=3}.HandlerDirection)
        ButtonNavigationCancel.OnClick().Subscribe(Handler_Navigation_ToPetSelector)
        ButtonNavigationOK.OnClick().Subscribe(TryPlacePet)
        #ButtonNavigationFight.OnClick().Subscribe(StartFight)

    SimonUICanBeDisplayed() : logic =
        if:
            MiniGameUI:=MaybeHexMiniGameUI?
            HexMiniGameSelectorUI := MaybeHexMiniGameSelectorUI?
        then:
            if(UIIsChallengeCanvas? or UIIsNavigationCanvas? or HexMiniGameSelectorUI.UIIsDisplayed?):
                return false
            else:
                return true
        return true

    GetUsedPets() : tuple([]hexagame_pet,[]hexagame_pet) =
        var AllAgentPets : []hexagame_pet = array{}
        var AllAIPets : []hexagame_pet = array{}
        for(Pet : AllHexaPlayerPets, Pet.IsBeingUsed?):
            set AllAgentPets = AllAgentPets + array{Pet}
        for(Pet : AllHexaAIPets, Pet.IsBeingUsed?):
            set AllAIPets = AllAIPets + array{Pet}
        return (AllAgentPets,AllAIPets)  

    GetAllAgentPets(Agent : agent) : []pet_data_table =
        var AllPets : []pet_data_table = array{}
        if(BakeryMachinesManager := MaybeBakeryMachinesManager?, Player:=player[Agent], Player.IsActive[]):
            set AllPets = BakeryMachinesManager.PlayerDataCache.AllPets
        return AllPets

    CheckIfPositionIsEmpty(Position : tuple(int,int))<transacts> : logic =
        if(MaybePet := AllHexBoxes[Position(0)][Position(1)], ValidPet := MaybePet?):
            return false
        else: 
            ##Print("Position {Position(0)}|{Position(1)} is empty, moving there")
            return true   

    MoveAgentFocus(Agent : agent, Direction : int) : void =
        TupleFocus := GetAgentFocus(Agent)
        var Row : int = TupleFocus(0)
        var Column : int = TupleFocus(1)
        case(Direction):
            0=> #DOWN
                set Row = Row - 1
                if(Row < 0):
                    set Row = 0
            1=> #UP
                set Row = Row + 1
                if(Row > HexaNumRows-1):
                    set Row = HexaNumRows-1
            2=> #LEFT
                set Column = Column - 1
                if(Column < 0):
                    set Column = 0
            3=> #RIGHT
                set Column = Column + 1
                if(Column > HexaNumColumns-1):
                    set Column = HexaNumColumns-1
            _=> #Select

        set AgentFocus= (Row,Column)

            for(IndexRow:=HexaMaxRowDesign+1..HexaNumRows-1, IndexColumn:=0..HexaNumColumns-1):
                if(HexaProp := AllHexPositions[IndexRow][IndexColumn]):
                    HexaProp.SetMaterial(Materials.MI_HexaBaseRed, ?Index:=2)

            
            for(IndexRow:=0..HexaMaxRowDesign, IndexColumn:=0..HexaNumColumns-1):
                if(HexaProp := AllHexPositions[IndexRow][IndexColumn]):
                    HexaProp.SetMaterial(Materials.MI_HexaBaseBlue, ?Index:=2)

            for(IndexRow:=0..HexaMaxRowDesign, IndexColumn:=0..HexaNumColumns-1):
                if(MaybePetHexBox := AllHexBoxes[IndexRow][IndexColumn], ValidPet := MaybePetHexBox?, Position := ValidPet.Position, HexaProp := AllHexPositions[Position(0)][Position(1)]):
                    HexaProp.SetMaterial(Materials.MI_HexaBaseGreen, ?Index:=2)

            if(HexaProp := AllHexPositions[Row][Column]):   
                HexaProp.SetMaterial(Materials.MI_HexaBaseGreen, ?Index:=2)

                var NormalScale : vector3 = vector3{}
                var PivotVariation : vector3 = vector3{}

                if(ValidPet := MaybeSelectedPet?):
                    if(ValidPet.Tier=1):
                        set NormalScale = PetScale_Tier1
                        set PivotVariation = PetPivotVariation_Tier1
                    if(ValidPet.Tier=2):
                        set NormalScale = PetScale_Tier2
                        set PivotVariation = PetPivotVariation_Tier2
                    if(ValidPet.Tier=3):
                        set NormalScale = PetScale_Tier3
                        set PivotVariation = PetPivotVariation_Tier3
        
                HexTransform:=HexaProp.GetTransform()
                Translation : vector3 = HexTransform.Translation+PivotVariation
                NewTransform:=transform{Translation:=Translation,Rotation:=HexaProp.GetTransform().Rotation,Scale:=NormalScale}

                if(SelectedPetProp.IsValid[], SelectedPetProp.TeleportTo[NewTransform]):

    #Intention to move
    

    CalculeBoxDistance(TuplePositionA : tuple(int,int), TuplePositionB : tuple(int,int)) : tuple(int,int)  =
        RowDistance:=Abs(TuplePositionA(0)-TuplePositionB(0))
        ColumnDistance:=Abs(TuplePositionA(1)-TuplePositionB(1))
        ##Print("Distance is [{RowDistance}][{ColumnDistance}]")
        return(RowDistance,ColumnDistance)

#PATH FINDERS
    #AI Path finder
    PathFinder(InstigatorPet : hexagame_pet, AllObjetives :[]hexagame_pet, Range : int, FindEnemies : logic) : tuple([]tuple(int,int),int,tuple(int,int)) =
        ##Print("#########################")

        var BestPathLength : int = -1
        var BestPath : []tuple(int,int) = array{(-1,-1),(-1,-1),(-1,-1),(-1,-1)}
        var ValidObjetivePosition : tuple(int,int) = (-1,-1)

        var MaybeCurrentObjetive : ?hexagame_pet = false
        var ValidPathFound_On1stLoop : logic = false
        var ValidPathFound_On2ndLoop : logic = false
        var ValidPathFound_On3thLoop : logic = false
        var ValidPathFound_On4thLoop : logic = false

        var IsObjetiveOnRange : logic = false
        #Detect if there is an valid objetive nearby
        for(Objetive : AllObjetives):
            if:
                ObjetiveProp := Objetive.MaybeProp?
                ValidInstigatorPetProp := InstigatorPet.MaybeProp?
                DistanceBetweenPets:=Floor[DistanceXY(ObjetiveProp.GetTransform().Translation,ValidInstigatorPetProp.GetTransform().Translation)]
            then:
                if:
                    Objetive.Health > 0.0
                    DistanceBetweenPets <= Range
                then:
                    set IsObjetiveOnRange = true
                    set BestPathLength = 0

        if(not IsObjetiveOnRange?):
            ###### FIRST LOOP ######
            #First loop, find on adjacent positions
            AllFirstLoopPosibleMovements := GetAllPosibleMovements(InstigatorPet.Position)  #Base position: StartingPosition
            #if(FindEnemies?, PetFaction = PLAYER_PET):
                ##Print("AllFirstLoopPosibleMovements: {AllFirstLoopPosibleMovements.Length}")
            for(Index_1:=0..AllFirstLoopPosibleMovements.Length-1, PosibleMovement_1stLoop := AllFirstLoopPosibleMovements[Index_1]):
                var CreatedPath : []tuple(int,int) = array{(-1,-1),(-1,-1),(-1,-1),(-1,-1)} #Empty the patch completly
                var CreatedPathLength : int = 9
                if(MaybeOtherHexaPet_1 := AllHexBoxes[PosibleMovement_1stLoop(0)][PosibleMovement_1stLoop(1)]):
                    if(ValidOtherPet_1 := MaybeOtherHexaPet_1?): #Is there a pet here or is empty?
                        IsValidObjetive_1 := CompareAndCheckPosition(InstigatorPet,0,BestPathLength,InstigatorPet.IsPlayerPet,FindEnemies,InstigatorPet.Position,PosibleMovement_1stLoop,ValidOtherPet_1,MaybeCurrentObjetive)  
                        if(IsValidObjetive_1?): #Set this objetive is better
                            set BestPathLength = 0
                            set BestPath = CreatedPath
                            set ValidObjetivePosition = PosibleMovement_1stLoop
                            set ValidPathFound_On1stLoop = true #Is there is an objetive valid nearby, we don´t need to create a second loop
                            set MaybeCurrentObjetive = MaybeOtherHexaPet_1
                            if:
                                set CreatedPath[0] = (-1,-1) #Saving this movement if there was not pets on the first 
                                set CreatedPath[1] = (-1,-1) #Saving this movement if there was not pets on the first 
                                set CreatedPath[2] = (-1,-1) #Saving this movement if there was not pets on the first 
                                set CreatedPath[3] = (-1,-1) #Saving this movement if there was not pets on the first 
                            then:
                                set BestPath = CreatedPath
                    else:
                        #There is no pet valid, is a posible place where we can move, continue to second loop
                        if:
                            not ValidPathFound_On1stLoop?
                            set CreatedPath[0] = PosibleMovement_1stLoop #Saving this movement if there was not pets on the first 
                            set CreatedPath[1] = (-1,-1) #Cleaning the other next positions because there where not a valid option
                            set CreatedPath[2] = (-1,-1) #Saving this movement if there was not pets on the first 
                            set CreatedPath[3] = (-1,-1) #Saving this movement if there was not pets on the first 
                        then:
                            ###### SECOND LOOP ######
                            #Second loop, find on adjacent positions
                            AllSecondLoopPosibleMovements := GetAllPosibleMovements(PosibleMovement_1stLoop) #Base position: PosibleMovement_1stLoop
                            for(Index_2:=0..AllSecondLoopPosibleMovements.Length-1, PosibleMovement_2ndLoop := AllSecondLoopPosibleMovements[Index_2]):
                                if(MaybeOtherHexaPet_2 := AllHexBoxes[PosibleMovement_2ndLoop(0)][PosibleMovement_2ndLoop(1)]):           
                                    if(ValidOtherPet_2 := MaybeOtherHexaPet_2?): #Is there a pet here or is empty?
                                        #Here there is a pet, is it is not empty
                                        IsValidObjetive_2:=CompareAndCheckPosition(InstigatorPet,1,BestPathLength,InstigatorPet.IsPlayerPet,FindEnemies,InstigatorPet.Position,PosibleMovement_2ndLoop,ValidOtherPet_2, MaybeCurrentObjetive)
                                        if(IsValidObjetive_2?): #Set this objetive is better
                                            set BestPathLength = 1
                                            set ValidObjetivePosition = PosibleMovement_2ndLoop
                                            set ValidPathFound_On2ndLoop = true #Is there is an objetive valid nearby, we don´t need to create a third loop
                                            set MaybeCurrentObjetive = MaybeOtherHexaPet_2
                                            if:
                                                set CreatedPath[1] = (-1,-1) #Cleaning the other next positions because there where not a valid option
                                                set CreatedPath[2] = (-1,-1) #Saving this movement if there was not pets on the first 
                                                set CreatedPath[3] = (-1,-1) #Saving this movement if there was not pets on the first 
                                            then:
                                                set BestPath = CreatedPath
                                    else:
                                        #There is no pet valid, is a posible place where we can move, continue to third and last loop
                                        if:
                                            not ValidPathFound_On2ndLoop?
                                            set CreatedPath[1] = PosibleMovement_2ndLoop #Saving this movement if there was not pets on the first 
                                            set CreatedPath[2] = (-1,-1) #Cleaning the other next positions because there where not a valid option
                                            set CreatedPath[3] = (-1,-1) #Saving this movement if there was not pets on the first 
                                        then:
                                            ###### THIRD LOOP ######
                                            #Third loop, find on adjacent positions
                                            AllThirdLoopPosibleMovements := GetAllPosibleMovements(PosibleMovement_2ndLoop) #Base position: PosibleMovement_2ndLoop
                                            for(Index_3:=0..AllThirdLoopPosibleMovements.Length-1, PosibleMovement_3thLoop := AllThirdLoopPosibleMovements[Index_3]):
                                                if(MaybeOtherHexaPet_3 := AllHexBoxes[PosibleMovement_3thLoop(0)][PosibleMovement_3thLoop(1)]):
                                                    if(ValidOtherPet_3 := MaybeOtherHexaPet_3?): #Is there a pet here or is empty?
                                                        #Here there is a pet, is it is not empty
                                                        IsValidObjetive_3:=CompareAndCheckPosition(InstigatorPet,2, BestPathLength, InstigatorPet.IsPlayerPet,FindEnemies,InstigatorPet.Position,PosibleMovement_3thLoop,ValidOtherPet_3, MaybeCurrentObjetive)
                                                        if(IsValidObjetive_3?): #Set this objetive is better
                                                            set BestPathLength = 2
                                                            set ValidObjetivePosition = PosibleMovement_3thLoop
                                                            set ValidPathFound_On3thLoop = true #Is there is an objetive valid nearby, we don´t need to create a third loop
                                                            set MaybeCurrentObjetive = MaybeOtherHexaPet_3
                                                            if:
                                                                set CreatedPath[2] = (-1,-1) #Saving this movement if there was not pets on the first 
                                                                set CreatedPath[3] = (-1,-1) #Saving this movement if there was not pets on the first 
                                                            then:
                                                                set BestPath = CreatedPath
                                                    else:
                                                        #There is no pet valid, is a posible place where we can move, continue to fourth and last loop
                                                        if:
                                                            not ValidPathFound_On3thLoop?
                                                            set CreatedPath[2] = PosibleMovement_3thLoop #Saving this movement if there was not pets on the first
                                                            set CreatedPath[3] = (-1,-1) #Saving this movement if there was not pets on the first 
                                                        then: 
                                                            ###### FOURTH LOOP ######
                                                            #Fourth loop, find on adjacent positions
                                                            AllFourthLoopPosibleMovements := GetAllPosibleMovements(PosibleMovement_3thLoop) #Base position: PosibleMovement_3thLoop
                                                            for(Index_4:=0..AllFourthLoopPosibleMovements.Length-1, PosibleMovement_4thLoop := AllFourthLoopPosibleMovements[Index_4]):
                                                                if(MaybeOtherHexaPet_4 := AllHexBoxes[PosibleMovement_4thLoop(0)][PosibleMovement_4thLoop(1)]):
                                                                    #if(PetFaction?):
                                                                        ##Print("Checking: [{PosibleMovement_4thLoop(0)}][{PosibleMovement_4thLoop(1)}]")
                                                                    if(ValidOtherPet_4 := MaybeOtherHexaPet_4?): #Is there a pet here or is empty?
                                                                        #Here there is a pet, is it is not empty
                                                                        IsValidObjetive_4:=CompareAndCheckPosition(InstigatorPet, 3, BestPathLength, InstigatorPet.IsPlayerPet,FindEnemies,InstigatorPet.Position,PosibleMovement_4thLoop,ValidOtherPet_4, MaybeCurrentObjetive)
                                                                        if(IsValidObjetive_4?): #Set this objetive is better
                                                                            ##Print("An objetive was found on the 4th loop on position: [{PosibleMovement_4thLoop(0)}],[{PosibleMovement_4thLoop(1)}]")
                                                                            set BestPathLength = 3
                                                                            set ValidObjetivePosition = PosibleMovement_4thLoop
                                                                            set ValidPathFound_On4thLoop = true #Is there is an objetive valid nearby, we don´t need to create a third loop
                                                                            set MaybeCurrentObjetive = MaybeOtherHexaPet_4
                                                                            if:
                                                                                set CreatedPath[3] = (-1,-1) #Saving this movement if there was not pets on the first 
                                                                            then:
                                                                                set BestPath = CreatedPath
                                                                else:
                                                                    PrintFatalError("On PathFinder AllHexBoxes[{PosibleMovement_4thLoop(0)}],[{PosibleMovement_4thLoop(1)}] not valid!")

                                                else:
                                                    PrintFatalError("On PathFinder AllHexBoxes[{PosibleMovement_3thLoop(0)}],[{PosibleMovement_3thLoop(1)}] not valid!")
                                else:
                                    PrintFatalError("On PathFinder AllHexBoxes[{PosibleMovement_2ndLoop(0)}],[{PosibleMovement_2ndLoop(1)}] not valid!")
                else:
                    PrintFatalError("On PathFinder AllHexBoxes[{PosibleMovement_1stLoop(0)}],[{PosibleMovement_1stLoop(1)}] not valid!")
            #Find enemy
                    <#
            if:
                Path0:=BestPath[0]
                Path1:=BestPath[1]
                Path2:=BestPath[2]
            then:
                #Print("#########################")
                #Print("Path0: [{Path0(0)}][{Path0(1)}]")
                #Print("Path1: [{Path1(0)}][{Path1(1)}]")
                #Print("Path2: [{Path2(0)}][{Path2(1)}]")
                #Print("#########################")#>
        return (BestPath,BestPathLength,ValidObjetivePosition)


    CompareAndCheckPosition<private>(InstigatorPet : hexagame_pet, OtherPetBoxDistance : int, CurrentPetBoxDistance : int ,PetFaction : logic, FindEnemies : logic, StartingPosition : tuple(int,int), PosibleMovement : tuple(int,int), OtherPet : hexagame_pet, MaybeCurrentValidObjetive : ?hexagame_pet) : logic = 
        var IsValid : logic = false
        if:
            PosibleMovement(0) = StartingPosition(0)
            PosibleMovement(1) = StartingPosition(1)
        then:
            return false
        else: # Position is different, so it is a different pet
            if:
                OtherPet.IsPlayerPet <> PetFaction
                OtherPet.Health > 0.0 #Is alive
                FindEnemies?
            then:
                if(CurrentValidPet := MaybeCurrentValidObjetive?):
                    if(OtherPetBoxDistance < CurrentPetBoxDistance or CurrentPetBoxDistance = -1):
                        set IsValid = true
                    else if(OtherPetBoxDistance = CurrentPetBoxDistance):
                        if(CurrentValidPet.Health>OtherPet.Health):
                            set IsValid = true
                    else:
                        #Print("Enemy not valid OtherPet.Health: {OtherPet.Health}, CurrentValidPet.Health: {CurrentValidPet.Health}")
                else:
                    set IsValid = true
            if:
                OtherPet.IsPlayerPet = PetFaction
                OtherPet.Health > 0.0 #Is alive
                not FindEnemies?
            then:
                if(CurrentValidPet := MaybeCurrentValidObjetive?):
                    if(OtherPetBoxDistance < CurrentPetBoxDistance or CurrentPetBoxDistance = -1):
                        set IsValid = true
                    else if(OtherPetBoxDistance = CurrentPetBoxDistance):
                        if(CurrentValidPet.Health>OtherPet.Health):
                            set IsValid = true
                else:
                    set IsValid = true
        #if(IsValid?, FindEnemies?, PetFaction = PLAYER_PET):
            ##Print("Valid enemy on [{PosibleMovement(0)}][{PosibleMovement(1)}], distance of movements: {OtherPetBoxDistance}")
        return IsValid

    #All posible movements
    GetAllPosibleMovements(StartingPosition : tuple(int,int)) : []tuple(int,int) =
        var AllPosibleMovements : []tuple(int,int) = array{}

        TupleOptionDownPositions:=GetNearDownPositions(StartingPosition)
        TupleOptionUpPositions:=GetNearUpPositions(StartingPosition)
        TupleOptionSidePositions:=GetNearSidePositions(StartingPosition)

        #Up
        if(StartingPosition<>TupleOptionUpPositions(0)):
            set AllPosibleMovements = AllPosibleMovements + array{TupleOptionUpPositions(0)}
        if(StartingPosition<>TupleOptionUpPositions(1)):
            set AllPosibleMovements = AllPosibleMovements + array{TupleOptionUpPositions(1)}
        #Down
        if(StartingPosition<>TupleOptionDownPositions(0)):
            set AllPosibleMovements = AllPosibleMovements + array{TupleOptionDownPositions(0)}
        if(StartingPosition<>TupleOptionDownPositions(1)):
            set AllPosibleMovements = AllPosibleMovements + array{TupleOptionDownPositions(1)}
        #Sides
        if(StartingPosition<>TupleOptionSidePositions(0)):
            set AllPosibleMovements = AllPosibleMovements + array{TupleOptionSidePositions(0)}
        if(StartingPosition<>TupleOptionSidePositions(1)):
            set AllPosibleMovements = AllPosibleMovements + array{TupleOptionSidePositions(1)}

        return AllPosibleMovements

    #Get near up positions
    GetNearUpPositions(Position : tuple(int,int)) : tuple(tuple(int,int),tuple(int,int)) =
        Row:=Position(0)
        Column:=Position(1)
        var UpRow : int = Row + 1
        var LeftColumn : int = Column
        var RightColumn : int = Column
        if(UpRow > HexaNumRows-1):
            set UpRow = HexaNumRows-1
        if(Row = 0 or Row = 2 or Row = 4 or Row = 6 or Row = 8):
            set LeftColumn = Column - 1
            if(LeftColumn < 0):
                set LeftColumn = 0
            set RightColumn = Column
        else if(Row = 1 or Row = 3 or Row = 5 or Row = 7 or Row = 9):
            set LeftColumn = Column
            set RightColumn = Column + 1
            if(RightColumn > HexaNumColumns-1):
                set RightColumn = HexaNumColumns-1
        return ((UpRow,LeftColumn),(UpRow,RightColumn))

    #Get near down positions
    GetNearDownPositions(Position : tuple(int,int)) : tuple(tuple(int,int),tuple(int,int)) =
        Row:=Position(0)
        Column:=Position(1)
        var DownRow : int = Row - 1
        var LeftColumn : int = Column
        var RightColumn : int = Column
        if(DownRow < 0):
            set DownRow = 0
        if(Row = 0 or Row = 2 or Row = 4 or Row = 6 or Row = 8):
            set LeftColumn = Column - 1
            if(LeftColumn < 0):
                set LeftColumn = 0
            set RightColumn = Column
        else if(Row = 1 or Row = 3 or Row = 5 or Row = 7 or Row = 9):
            set LeftColumn = Column
            set RightColumn = Column + 1
            if(RightColumn > HexaNumColumns-1):
                set RightColumn = HexaNumColumns-1
        return ((DownRow,LeftColumn),(DownRow,RightColumn))

    #Get near side positions
    GetNearSidePositions(Position : tuple(int,int)) : tuple(tuple(int,int),tuple(int,int)) =
        var Row : int = Position(0)
        var Column : int = Position(1)
        var ToTheLeftColumn : int = Position(1) - 1
        var ToTheRightColumn : int = Position(1) + 1
        if(ToTheLeftColumn<0):
            set ToTheLeftColumn = 0
        if(ToTheRightColumn>HexaNumColumns-1):
            set ToTheRightColumn = HexaNumColumns-1
        return ((Row,ToTheLeftColumn),(Row,ToTheRightColumn))

    SetAgent(Agent : agent) : void =
        Remove_NavigationCanvasWithoutMessage()
        Remove_ChallengeCanvasWithoutMessage()

        if(BakeryMachinesManager:=MaybeBakeryMachinesManager?):
            set ChallengeCurrentSelection = BakeryMachinesManager.PlayerDataCache.HexaGame_ChallengePointer
        set GameIsRunning = false
        set IsUsingHexaGame = false
        ResetLastGame(true)
        set MaybeAgent = option{Agent}

        if(HexMiniGameUI := MaybeHexMiniGameUI?):
            HexMiniGameUI.SetAgent(Agent)
        if(HexMiniGameSelectorUI := MaybeHexMiniGameSelectorUI?):
            HexMiniGameSelectorUI.SetAgent(Agent)

    GetAgentFocus(Agent : agent) : tuple(int,int) =
        return AgentFocus

    SelectChallenge(Player : player) : void =
        if(Agent := agent[Player]):
            #Set Valid movements
            for(IndexRow:=0..HexaMaxRowDesign, IndexColumn:=0..HexaNumColumns-1):
                if(HexaProp := AllHexPositions[IndexRow][IndexColumn]):
                    HexaProp.SetMaterial(Materials.MI_HexaBaseBlue, ?Index:=2)

            for(IndexRow:=HexaMaxRowDesign+1..HexaNumRows-1, IndexColumn:=0..HexaNumColumns-1):
                if(HexaProp := AllHexPositions[IndexRow][IndexColumn]):
                    HexaProp.SetMaterial(Materials.MI_HexaBaseRed, ?Index:=2)     

    ShowChallenges(Agent : agent) : void =
        #First, check if the player has pets
        if(BakeryMachinesManager:=MaybeBakeryMachinesManager?):
            if(AgentOwner:=MaybeAgent?, PlayerOwner:=player[AgentOwner], Player := player[Agent], Player = PlayerOwner):
                if(Player.IsActive[], BakeryMachinesManager.PlayerDataCache.AllPets.Length > 0):
                    if(BakeryMachinesManager.PlayerDataCache.HexaGame_Accumulated>=0):
                        ResetLastGame(true)
                        if(ValidAgent:=MaybeAgent?):
                            if(Agent=ValidAgent):
                                BakeryMachinesManager.GameManager.VFXManager.StopAllMusics(Agent)
                                SetAgent(Agent)
                                LoadAgentPets(Agent)
                                Display_ChallengeCanvas()
                                RefreshChallengeUI()
                        set IsUsingHexaGame = true
                    else:
                        BakeryMachinesManager.GameManager.DisplayCharacterMessage(Agent, I_CHARACTER,"You have no Sourdough Competition Tokens", ?Sound := ERROR_SOUND)
                else:
                    BakeryMachinesManager.GameManager.DisplayCharacterMessage(Agent, I_CHARACTER,"You have no Sourdoughs", ?Sound := ERROR_SOUND)
            else:
                BakeryMachinesManager.GameManager.DisplayCharacterMessage(Agent, I_CHARACTER,"YOU ARE NO THE OWNER", ?Sound := ERROR_SOUND)
        else:
            PrintFatalError("MaybeBakeryMachinesManager not found!")

    Delay_ToPetSelector()<suspends> : void =
        Sleep(0.1)
        Display_SelectorCanvas()

    Display_SelectorCanvas(): void =
        if(Agent:=MaybeAgent?):
            CameraFixedPoint_Hexa.AddTo(Agent)
        if(HexMiniGameSelectorUI:=MaybeHexMiniGameSelectorUI?):
            spawn:
                HexMiniGameSelectorUI.Delay_Canvas()    

    CheckGameEnd(IsTimeOver : logic) : void =
        if(GameIsRunning?):
            #Print("Checking end!")
            var IsThereEnemies : logic = false
            var IsThereFriends : logic = false
            var IsGameWon : logic = false
            for(IndexRow:=0..HexaNumRows-1):
                for(IndexColumn:=0..HexaNumColumns-1):
                    if(MaybeHexPet := AllHexBoxes[IndexRow][IndexColumn], ValidPet := MaybeHexPet?, ValidPet.IsBeingUsed? ,ValidPet.Health>0.0):
                        if(ValidPet.IsPlayerPet?):
                            set IsThereFriends = true
                        else:
                            set IsThereEnemies = true
            if(not IsTimeOver?):
                if(IsThereFriends?, not IsThereEnemies?, ValidUI := MaybeHexMiniGameUI?):
                    ##Print("Ending game because is not enemies")
                    set IsGameWon = true
                    set GameIsRunning = false
                else if(not IsThereFriends?, IsThereEnemies?, ValidUI := MaybeHexMiniGameUI?):
                    ##Print("Ending game because is not allies")
                    set IsGameWon = false
                    set GameIsRunning = false
                else if(not IsThereFriends?, not IsThereEnemies?, ValidUI := MaybeHexMiniGameUI?):
                    ##Print("Ending game because is not enemies or allies")
                    set IsGameWon = false
                    set GameIsRunning = false
            else:
                ##Print("Ending game because time is over.")
                set GameIsRunning = false
                set IsGameWon = false
            if(not GameIsRunning?):
                SaveGameResult(IsGameWon)

                
    CalculateGatherReward(NumChallenge : int, FirstTime : logic, IsWon : logic, TimeRemaining : int) : float =
        var GatherReward : float = 0.0
        if(ValidBakery:=MaybeBakeryMachinesManager?):
            Log("Num Challenge: {NumChallenge}", LOG_LEVEL_DEBUG10)
            Proporcion := HexaGame_DelayTime_Base/Dungeon_DelayTime_Base
            Log("Gather Proporcion: {GetOptimizedMagnitudeLetter(Proporcion)}", LOG_LEVEL_DEBUG10)
            var MaxNumHitsGather : float = ValidBakery.GameManager.GeneralMobsManager.DungeonDataBase.GetMainChestMultiplier(6)+ValidBakery.GameManager.GeneralMobsManager.DungeonDataBase.GetSecondaryChestMultiplier(6)
            Log("MaxNumHitsGather: {GetOptimizedMagnitudeLetter(MaxNumHitsGather)}", LOG_LEVEL_DEBUG10)
            set MaxNumHitsGather = MaxNumHitsGather * Proporcion
            LevelDifference := HexaChallengesDatabase.AllChallenges.Length-1-NumChallenge
            Log("LevelDifference from MaxChallenge: {LevelDifference}", LOG_LEVEL_DEBUG10)
            Log("Final Total Gather: {GetOptimizedMagnitudeLetter(MaxNumHitsGather)}", LOG_LEVEL_DEBUG10)

            var FinalGather : float = MaxNumHitsGather-(LevelDifference*0.5)

            if(FinalGather<25.0):
                set FinalGather = 25.0
            set GatherReward = FinalGather
            Log("Gather Reward: {GetOptimizedMagnitudeLetter(GatherReward)}", LOG_LEVEL_DEBUG10)

            
            if(IsWon?):
                if(FirstTime?):
                    set GatherReward = GatherReward * HEXAGAME_FirstTimeMultiplier
            else:
                if(TimeRemaining<20):
                    set GatherReward = GatherReward*0.5
                else if(TimeRemaining<30):
                    set GatherReward = GatherReward*0.25
                else:
                    set GatherReward = GatherReward*0.10
        return GatherReward

    CalculeMutantMaterialGatherReward(NumChallenge : int, FirstTime : logic, IsWon : logic, NumHits : float) : float =
        var MutantMaterialAmount : float = 0.0
        CalculatedHBLevel:=CalculeHBLevel(NumChallenge)
        set MutantMaterialAmount = CheckFinalMutantMaterialAward(CalculatedHBLevel,NumHits,1.0)
        if(IsWon?):
            if(FirstTime?):
                set MutantMaterialAmount = MutantMaterialAmount * HEXAGAME_FirstTimeMultiplier
        else:
            set MutantMaterialAmount = MutantMaterialAmount * 0.35     
        return MutantMaterialAmount

    CalculeHBLevel<private>(NumChallenge : int) : int =
        var HBLevel : int = 0
        NumHexaChallengesPerLevel:=((HexaChallengesDatabase.AllChallenges.Length-1)*1.0)/(HB_MAXLEVEL*1.0)
        if(CalculatedHBLevel := Ceil[(NumChallenge*1.0)/NumHexaChallengesPerLevel]):
            set HBLevel = CalculatedHBLevel
        return HBLevel

    CalculeCoalGatherReward(NumChallenge : int, FirstTime : logic, IsWon : logic) : float =
        var CoalReward : float = 5.0
        Proportion_Multiplier:=HexaGame_DelayTime_Base/Dungeon_DelayTime_Base
        MaxHexaCoal:=HexaAndDungeons_MaxCoal*Proportion_Multiplier

        var HBLevel : int = 0
        if(ProgressPercentage := Ceil[(NumChallenge*1.0)/((HexaChallengesDatabase.AllChallenges.Length-1)*1.0)], AuxCoalReward := Ceil[ProgressPercentage*MaxHexaCoal]*1.0, AuxCoalReward > 5.0):
            set CoalReward = CoalReward
    
        if(IsWon?):
            if(FirstTime?):
                set CoalReward = CoalReward * HEXAGAME_FirstTimeMultiplier
        else if(AuxCoalReward:=Floor[CoalReward* 0.5]):
            set CoalReward = AuxCoalReward*1.0
            if(CoalReward < 5.0):
                set CoalReward = 3.0
        
        return CoalReward


    SaveGameResult(GameWon : logic) : void =
        if(BakeryMachinesManager := MaybeBakeryMachinesManager?, ValidUI := MaybeHexMiniGameUI?, Agent := MaybeAgent?, Player := player[Agent], Player.IsActive[]):
            var RandomRes1 : int = MUTANTMATERIAL
            var RandomRes2 : int = COAL
            var RandomRes3 : int = WATER
            var RandomRes4 : int = FLOUR
            var RandomRes5 : int = CHEESE
            var RandomRes6 : int = PIMENTOS

            var AmountRes1 : float = 50.0
            var AmountRes2 : float = 50.0
            var AmountRes3 : float = 50.0
            var AmountRes4 : float = 50.0
            var AmountRes5 : float = 50.0
            var AmountRes6 : float = 50.0

            HBLevel:=CalculeHBLevel(ChallengeCurrentSelection)
           
            var UpdatedDataTable : player_data_table = BakeryMachinesManager.GameManager.QuestsAndOrdersDatabase.AddAgentQuest_ProgressV2(Player,BakeryMachinesManager.PlayerDataCache,1.0,WINHEXAGAME_TUTORIAL_TYPE)(4)
            

            var IsFirstTimeCompleted : logic = false

            if(GameWon?):
                spawn:
                    BakeryMachinesManager.GameManager.VFXManager.PlaySound(WIN_SOUND,0.0,option{Agent},false)
                RegisterDataTuple := PlayersDataManager.RegisterHexaChallengeCompleted(UpdatedDataTable,ChallengeCurrentSelection)
                set UpdatedDataTable = RegisterDataTuple(0)
                set UpdatedDataTable = PlayersDataManager.SetChallengePointer(UpdatedDataTable,ChallengeCurrentSelection)
                set IsFirstTimeCompleted = RegisterDataTuple(1)  
            else:
                spawn:
                    BakeryMachinesManager.GameManager.VFXManager.PlaySound(LOSE_SOUND,0.0,option{Agent},false)
            
            GatherNumHits := CalculateGatherReward(ChallengeCurrentSelection,IsFirstTimeCompleted,GameWon,ValidUI.TotalGameTime)
            ResourcersToGive:=BakeryMachinesManager.GetBakeryPickaxeDamage(HBLevel,false,true)(0)*GatherNumHits

            #SET NEW MATERIALS
            set RandomRes1 = MUTANTMATERIAL
            set RandomRes2 = COAL
            set RandomRes3 = WATER
            set RandomRes4 = FLOUR
            set RandomRes5 = CHEESE
            set RandomRes6 = PIMENTOS

            set AmountRes1 = CalculeMutantMaterialGatherReward(ChallengeCurrentSelection,IsFirstTimeCompleted,GameWon,GatherNumHits)
            set AmountRes2 = CalculeCoalGatherReward(ChallengeCurrentSelection,IsFirstTimeCompleted,GameWon)
            set AmountRes3 = ResourcersToGive
            set AmountRes4 = ResourcersToGive
            set AmountRes5 = ResourcersToGive
            set AmountRes6 = ResourcersToGive

            Log("Mutant Material Granted: {AmountRes1}",LOG_LEVEL_DEBUG7)
            Log("Coal Granted: {AmountRes2}",LOG_LEVEL_DEBUG7)
            Log("Water Granted: {AmountRes3}",LOG_LEVEL_DEBUG7)
            Log("Flour Granted: {AmountRes4}",LOG_LEVEL_DEBUG7)
            Log("Cheese Material Granted: {AmountRes5}",LOG_LEVEL_DEBUG7)
            Log("Pimentos Granted: {AmountRes6}",LOG_LEVEL_DEBUG7)

            AllAmounts : []float := array{AmountRes3,AmountRes4,AmountRes5,AmountRes6,AmountRes1,AmountRes2}
            AllTypes : []int := array{WATER,FLOUR,CHEESE,PIMENTOS,MUTANTMATERIAL,COAL}

            BakeryMachinesManager.SetNewPlayerDataCache(UpdatedDataTable)
            BakeryMachinesManager.SaveResourcers(Agent,AllTypes,AllAmounts, ?AutoBackpack:=true)
            BakeryMachinesManager.GameManager.DisplayResourcersGather(Agent,AllAmounts,AllTypes,false)
            spawn:
                BakeryMachinesManager.GameManager.CalculateAndSaveFame(Agent)
            if(HexMiniGameUI:=MaybeHexMiniGameUI?):
                HexMiniGameUI.DisplayResult(array{AmountRes1,AmountRes2,AmountRes3,AmountRes4,AmountRes5,AmountRes6},array{RandomRes1,RandomRes2,RandomRes3,RandomRes4,RandomRes5,RandomRes6},GameWon)
        else:
            PrintFatalError("On SaveGameResult: Player is Invalid")

    ResetAgentPets() : void =
        for(HexaPet : AllHexaPlayerPets):
            if(BakeryMachinesManager := MaybeBakeryMachinesManager?, var AllPlayerPets : []pet_data_table = BakeryMachinesManager.PlayerDataCache.AllPets,var PetHexaUsed : pet_data_table = AllPlayerPets[HexaPet.AllPetsIndex]):
                set PetHexaUsed = BakeryMachinesManager.PlayersDataManager.SetHexaGamePosition(PetHexaUsed,-1,-1)
                if(set AllPlayerPets[HexaPet.AllPetsIndex] = PetHexaUsed):
                    NewPlayerDataTable:=BakeryMachinesManager.PlayersDataManager.SavePets(BakeryMachinesManager.PlayerDataCache,AllPlayerPets)
                    BakeryMachinesManager.SetNewPlayerDataCache(NewPlayerDataTable)
            HexaPet.End()
        if(Agent := MaybeAgent?, Player := player[Agent], Player.IsActive[]):
            LoadAgentPets(Agent)
            for(IndexRow:=0..HexaNumRows-1, IndexColumn:=0..HexaNumColumns-1):
                if(HexaProp := AllHexPositions[IndexRow][IndexColumn]):
                    HexaProp.SetMaterial(Materials.MI_HexaBaseNoFresnel, ?Index:=2)
        for(IndexRow:=0..HexaNumRows-1):
            for(IndexColumn:=0..HexaNumColumns-1):
                if:
                    MaybePet:=AllHexBoxes[IndexRow][IndexColumn]
                    ValidPet:=MaybePet?
                    ValidPet.IsPlayerPet?
                    set AllHexBoxes[IndexRow][IndexColumn] = false

    ResetLastGame(IsCompleteReset : logic, ?PlayerLeftGame : logic = false) : void =
        set GameIsRunning = false
        set MaybeSelectedPet = false
        for(IndexRow:=0..HexaNumRows-1):
            for(IndexColumn:=0..HexaNumColumns-1):
                if(set AllHexBoxes[IndexRow][IndexColumn] = false):

        set AgentFocus = (0,0)
        for(Pet : AllHexaAIPets):
            Pet.End()
            #Pet.SetIsBeingUsed(false)
        for(Pet : AllHexaPlayerPets):
            Pet.End()
            #et.SetIsBeingUsed(false)

        if(not  PlayerLeftGame?, Agent := MaybeAgent?, Player := player[Agent], Player.IsActive[]):
            #if(IsCompleteReset?):
            LoadAgentPets(Agent)

        for(IndexRow:=0..HexaNumRows-1, IndexColumn:=0..HexaNumColumns-1):
            if(HexaProp := AllHexPositions[IndexRow][IndexColumn]):
                HexaProp.SetMaterial(Materials.MI_HexaBaseNoFresnel, ?Index:=2)

        ##########
        #Init Matrix
        for(IndexRow:=0..HexaNumRows-1):
            for(IndexColumn:=0..HexaNumColumns-1):
                if:
                    set AllHexBoxes[IndexRow][IndexColumn] = false
                #then:
                        ##Print("Cleaned position [{IndexRow}][{IndexColumn}]")


    Display_HexMiniGameSelectorUI() : void =
        if(MiniGameSelector := MaybeHexMiniGameSelectorUI?):
            spawn:
                MiniGameSelector.Delay_Canvas()


    LoadAgentPets(Agent : agent) : void =
        for(HexaPlayerPet : AllHexaPlayerPets):
            HexaPlayerPet.End()
        set AllHexaPlayerPets = array{}
        if(Player := player[Agent], Player.IsActive[], BakeryMachinesManager := MaybeBakeryMachinesManager?):
            #CreateHexPets
            for(Index := 0 ..BakeryMachinesManager.PlayerDataCache.AllPets.Length-1, Pet := BakeryMachinesManager.PlayerDataCache.AllPets[Index]):
                
                HexaGamePositionX := Pet.HexaGamePositionX
                HexaGamePositionY := Pet.HexaGamePositionY

                TuplePurity := ConvertRawPurity(Pet.Purity)

                MaxHealth := CalculePetStat(STAT_TYPE_HEALTH,Pet.LVPet_PackageA,Pet.Type,Pet.Purity)(0)
                AttackDamage := CalculePetStat(STAT_TYPE_DAMAGE,Pet.LVPet_PackageA,Pet.Type,Pet.Purity)(0)
                Speed := CalculePetStat(STAT_TYPE_SPEED,Pet.LVPet_PackageA,Pet.Type,Pet.Purity)(0)
                
                HealingPerk := CalculePetStat(STAT_TYPE_HEALING,Pet.LVPet_PackageB,Pet.Type,Pet.Purity)(0)
                AttackSpeed := CalculePetStat(STAT_TYPE_ATTACKSPEED,Pet.LVPet_PackageB,Pet.Type,Pet.Purity)(0)

                AttackRange:=ConvertRange(true, CalculePetStat(STAT_TYPE_RANGE,Pet.LVPet_PackageB,Pet.Type,Pet.Purity)(0))
                DefendRange:=ConvertRange(true, CalculePetStat(STAT_TYPE_RANGE,Pet.LVPet_PackageB,Pet.Type,Pet.Purity)(0))

                ##Print("AttackRange: {AttackRange}, TuplePurity: Tier {TuplePurity(1)} Purity {TuplePurity(0)}")
       
                if:
                    IntAttackRange := Floor[AttackRange]
                    IntDefendRange := Floor[DefendRange]
                   
                then:
                    HexaPetTable := MakeHexaPetTable(Index,option{Self},BakeryMachinesManager.GameManager.PropsManager.GetPetProp(Pet.Type),
                        BakeryMachinesManager.GameManager.VFXManager.GetMeleeAttackProp(),BakeryMachinesManager.GameManager.VFXManager.GetRangeAttackProp(),BakeryMachinesManager.GameManager.VFXManager.GetDefendProp(), Pet.LVPet_PackageA, Pet.LVPet_PackageB,IntAttackRange, IntDefendRange,
                        Pet.Type, PLAYER_PET,TuplePurity(0), TuplePurity(1), MaxHealth, AttackDamage, Speed, 
                        AttackSpeed, HealingPerk)
                    HexaPetTable.SetPosition((-1,-1))
                    set AllHexaPlayerPets = AllHexaPlayerPets + array{HexaPetTable} 
                    if(HexaGamePositionX <> -1, HexaGamePositionY <> -1):
                        Log("Pet assigned before detected, Index: {Index}",LOG_LEVEL_DEBUG6)
                        var NumUsedPets : int = 0
                        for(HexaPlayerPet : AllHexaPlayerPets):
                            if(HexaPlayerPet.IsBeingUsed?):
                                set NumUsedPets = NumUsedPets + 1
                        NumPetsPlaced.SetText(BasicMessage("{NumUsedPets}/5 used slots"))
                        UpdateNavigationCanvas()
                        if(NumUsedPets<5):
                            PlacePet(HexaPetTable,((HexaGamePositionX),HexaGamePositionY))  


              <#      
    ReloadPlayerPets() : void =
        #CreateHexPets
        for(PlayerPet : AllHexaPlayerPets):
            PlayerPet.SetPosition(PlayerPet.InitialPosition)
            PlayerPet.InitHexaPet() #>      

    <#
    StartHexGame()<suspends> : void =
        if(TestsManager := MaybeTestManager?, Agent:=MaybeAgent?, Player:=player[Agent], Player.IsActive[], PlayerData:=PlayersDataManager.GetPlayerData[Player]):
            PlayerDataWithoutToken:=PlayersDataManager.RemoveAccumulatedHexaChallengeToken(PlayerData)
            PlayersDataManager.SaveNewData(Player,PlayerData,PlayerDataWithoutToken)
            TestsManager.StartChallenge_AudioPlayer.Play(Agent)
            for(IndexRow:=0..HexaNumRows-1):
                for(IndexColumn:=0..HexaNumColumns-1):
                    if(HexaProp := AllHexPositions[IndexRow][IndexColumn]):
                        HexaProp.SetMaterial(Materials.MI_HexaBaseNoFresnel, ?Index:=2)

            #Remove this code in the future
            for(AIPet : AllHexaAIPets, AIPet.IsBeingUsed?):
                AIPet.CreateID()
            for(PlayerPet : AllHexaPlayerPets, PlayerPet.IsBeingUsed?):
                PlayerPet.CreateID()
            ##############
            Sleep(0.5)
            #First, save AI positions to avoid being in the same position
            for(AIPet : AllHexaAIPets, AIPet.IsBeingUsed?):
                AIPet.SetPosition(AIPet.Position(0),AIPet.Position(1))
            Sleep(1.0)
            for(PlayerPet : AllHexaPlayerPets, PlayerPet.IsBeingUsed?):
                spawn:
                    PlayerPet.RunPet()
            for(AIPet : AllHexaAIPets, AIPet.IsBeingUsed?):
                spawn:
                    AIPet.RunPet()
            if(HexaMinigameManagerUI := MaybeHexMiniGameUI?):
                HexaMinigameManagerUI.Display_Canvas()
                HexaMinigameManagerUI.CleanUI()
                HexaMinigameManagerUI.RefreshAll()
                HexaMinigameManagerUI.StartTime()
        #Remove this checker in the future
        loop:
            Sleep(0.25)
            for(PetToCheck : AllHexaPlayerPets, PetToCheck.IsBeingUsed?):
                PositionToCheck:=PetToCheck.Position
                for(OtherPlayerPet : AllHexaPlayerPets, OtherPlayerPet.IsBeingUsed?):
                    if(PositionToCheck=OtherPlayerPet.Position, PetToCheck.ID <> OtherPlayerPet.ID, GameIsRunning?, OtherPlayerPet.Health>0.0):
                        PrintFatalError("There is two pets on position [{PositionToCheck(0)}][{PositionToCheck(1)}]")
                for(OtherAIPet : AllHexaAIPets, OtherAIPet.IsBeingUsed?, GameIsRunning?):
                    if(PetToCheck.Position=OtherAIPet.Position, PetToCheck.ID <> OtherAIPet.ID, OtherAIPet.Health>0.0):
                        PrintFatalError("There is two pets on position [{PositionToCheck(0)}][{PositionToCheck(1)}]")
            for(PetToCheck : AllHexaAIPets, PetToCheck.IsBeingUsed?):
                PositionToCheck:=PetToCheck.Position
                for(OtherPlayerPet : AllHexaPlayerPets, OtherPlayerPet.IsBeingUsed?):
                    if(PositionToCheck=OtherPlayerPet.Position, PetToCheck.ID <> OtherPlayerPet.ID, OtherPlayerPet.Health>0.0):
                        PrintFatalError("There is two pets on position [{PositionToCheck(0)}][{PositionToCheck(1)}]")
                for(OtherAIPet : AllHexaAIPets, OtherAIPet.IsBeingUsed?):
                    if(PetToCheck.Position=OtherAIPet.Position, PetToCheck.ID <> OtherAIPet.ID, OtherAIPet.Health>0.0):
                        PrintFatalError("There is two pets on position [{PositionToCheck(0)}][{PositionToCheck(1)}]")
            if(not GameIsRunning?):
                break#>

    StartHexGame()<suspends> : void =
        if(Agent:=MaybeAgent?, Player:=player[Agent], Player.IsActive[], BakeryMachinesManager := MaybeBakeryMachinesManager?):
            PlayerDataWithoutToken:=PlayersDataManager.RemoveAccumulatedHexaChallengeToken(BakeryMachinesManager.PlayerDataCache)
            BakeryMachinesManager.SetNewPlayerDataCache(PlayerDataWithoutToken)
            BakeryMachinesManager.DisplayResourcersGather(array{-1.0},array{HEXATOKEN},false)
            #BakeryMachinesManager.GameManager.VFXManager.PlayerPays_AudioPlayer.Play(Agent)
            
            #PlayersDataManager.SaveAllNewData(Player,PlayerData,PlayerDataWithoutToken, BakeryMachinesManager)
            #PlayersDataManager.SaveNewData(Player,PlayerData,PlayerDataWithoutToken)
            spawn:
                BakeryMachinesManager.GameManager.VFXManager.PlaySound(STARTCHALLENGE_SOUND,0.0,option{Agent},false)
            for(IndexRow:=0..HexaNumRows-1):
                for(IndexColumn:=0..HexaNumColumns-1):
                    if(HexaProp := AllHexPositions[IndexRow][IndexColumn]):
                        HexaProp.SetMaterial(Materials.MI_HexaBaseNoFresnel, ?Index:=2)
            ##############
            Sleep(0.5)
            #First, save AI positions to avoid being in the same position
            for(AIPet : AllHexaAIPets, AIPet.IsBeingUsed?):
                AIPet.SetPosition(AIPet.Position(0),AIPet.Position(1))
            for(PlayerPet : AllHexaPlayerPets, PlayerPet.IsBeingUsed?):
                PlayerPet.SetPosition(PlayerPet.Position(0),PlayerPet.Position(1))
            Sleep(1.0)
            spawn:
                HexGameLoop()
            if(HexaMinigameManagerUI := MaybeHexMiniGameUI?):
                HexaMinigameManagerUI.Display_Canvas()
                HexaMinigameManagerUI.CleanUI()
                HexaMinigameManagerUI.RefreshAll()
                HexaMinigameManagerUI.StartTime()
            

    #New code
    HexGameLoop()<suspends> : void =
        ##Print("Starting game")
        for(PlayerPet : AllHexaPlayerPets, PlayerPet.IsBeingUsed?):
            PlayerPet.InitPet()
            if(PlayerPet.IsMoving?):
                PrintFatalError("Runing a pet, this pet is already moving before game: PetType {PlayerPet.PetType} | Position [{PlayerPet.Position(0)}][{PlayerPet.Position(1)}]")
            if(PlayerPet.Health<=0.0):
                PrintFatalError("Runing a pet, this pet is dead: PetType {PlayerPet.PetType} | Position [{PlayerPet.Position(0)}][{PlayerPet.Position(1)}]")
        for(AIPet : AllHexaAIPets, AIPet.IsBeingUsed?):
            AIPet.InitPet()
            if(AIPet.IsMoving?):
                PrintFatalError("Runing a pet, this pet is already moving before game: PetType {AIPet.PetType} | Position [{AIPet.Position(0)}][{AIPet.Position(1)}]")
            if(AIPet.Health<=0.0):
                PrintFatalError("Runing a pet, this pet is dead: PetType {AIPet.PetType} | Position [{AIPet.Position(0)}][{AIPet.Position(1)}]")
        #Remove this checker in the future
        loop:
            Sleep(0.04)
            for(AIPet : AllHexaAIPets, AIPet.IsBeingUsed?):
                if(AIPet.Health<=0.0):
                    AIPet.Die()
                else:
                    AIPet.UseTurn()
            for(PlayerPet : AllHexaPlayerPets, PlayerPet.IsBeingUsed?):
                if(PlayerPet.Health<=0.0):
                    PlayerPet.Die()
                else:
                    PlayerPet.UseTurn()
            #CheckEndGame()
            if(not GameIsRunning?):
                ##Print("Game ended")
                break
            
            
            
<#
    CheckEndGame<private>() : void = 
        for(PetToCheck : AllHexaPlayerPets, PetToCheck.IsBeingUsed?):
            PositionToCheck:=PetToCheck.Position
            for(OtherPlayerPet : AllHexaPlayerPets, OtherPlayerPet.IsBeingUsed?):
                if(PositionToCheck=OtherPlayerPet.Position, PetToCheck.ID <> OtherPlayerPet.ID, GameIsRunning?, OtherPlayerPet.Health>0.0):
                    PrintFatalError("There is two pets on position [{PositionToCheck(0)}][{PositionToCheck(1)}]")
            for(OtherAIPet : AllHexaAIPets, OtherAIPet.IsBeingUsed?, GameIsRunning?):
                if(PetToCheck.Position=OtherAIPet.Position, PetToCheck.ID <> OtherAIPet.ID, OtherAIPet.Health>0.0):
                    PrintFatalError("There is two pets on position [{PositionToCheck(0)}][{PositionToCheck(1)}]")
        for(PetToCheck : AllHexaAIPets, PetToCheck.IsBeingUsed?):
            PositionToCheck:=PetToCheck.Position
            for(OtherPlayerPet : AllHexaPlayerPets, OtherPlayerPet.IsBeingUsed?):
                if(PositionToCheck=OtherPlayerPet.Position, PetToCheck.ID <> OtherPlayerPet.ID, OtherPlayerPet.Health>0.0):
                    PrintFatalError("There is two pets on position [{PositionToCheck(0)}][{PositionToCheck(1)}]")
            for(OtherAIPet : AllHexaAIPets, OtherAIPet.IsBeingUsed?):
                if(PetToCheck.Position=OtherAIPet.Position, PetToCheck.ID <> OtherAIPet.ID, OtherAIPet.Health>0.0):
                    PrintFatalError("There is two pets on position [{PositionToCheck(0)}][{PositionToCheck(1)}]")#>

    OnPlayerRemoved() : void =
        #To do complete reset!
        SelectedPetProp.Dispose()
        set GameIsRunning = false

        set MaybeAgent = false

        set UIIsChallengeCanvas = false
        set UIIsNavigationCanvas = false

        Remove_ChallengeCanvasWithoutMessage() 
        Remove_NavigationCanvasWithoutMessage() 

        set MaybeSelectedPet = false

        ResetLastGame(true,?PlayerLeftGame := true)
        set AgentFocus = (0,0)

        if(HexMiniGameUI:=MaybeHexMiniGameUI?,HexMiniGameSelectorUI:=MaybeHexMiniGameSelectorUI?):
            HexMiniGameUI.Remove_CanvasWithoutMessage()
            HexMiniGameSelectorUI.Remove_CanvasWithoutMessage()

        set ChallengeCurrentSelection = 0

        CloseAllUI()

    SelectPet(CurrentOption : int) : void =
        set AgentFocus=(0,0)
        if(Agent:=MaybeAgent?):
            MoveAgentFocus(Agent,2)
        ##Print("Selecting option: {SelectPet}")

        for(IndexRow:=0..HexaMaxRowDesign, IndexColumn:=0..HexaNumColumns-1):
            if(HexaProp := AllHexPositions[IndexRow][IndexColumn]):
                HexaProp.SetMaterial(Materials.MI_HexaBaseBlue, ?Index:=2)

        for(IndexRow:=0..HexaMaxRowDesign, IndexColumn:=0..HexaNumColumns-1):
            if(MaybePetHexBox := AllHexBoxes[IndexRow][IndexColumn], ValidPet := MaybePetHexBox?, Position := ValidPet.Position, HexaProp := AllHexPositions[Position(0)][Position(1)]):
                HexaProp.SetMaterial(Materials.MI_HexaBaseGreen, ?Index:=2)
        if(HexaPet:=AllHexaPlayerPets[CurrentOption], HexaPropPosition := AllHexPositions[AgentFocus(0)][AgentFocus(1)]):

            var NormalScale : vector3 = vector3{}
            var PivotVariation : vector3 = vector3{}

            if(ValidPet := MaybeSelectedPet?):
                if(ValidPet.Tier=1):
                    set NormalScale = PetScale_Tier1
                    set PivotVariation = PetPivotVariation_Tier1
                if(ValidPet.Tier=2):
                    set NormalScale = PetScale_Tier2
                    set PivotVariation = PetPivotVariation_Tier2
                if(ValidPet.Tier=3):
                    set NormalScale = PetScale_Tier3
                    set PivotVariation = PetPivotVariation_Tier3
    
            HexTransform:=HexaPropPosition.GetTransform()
            Translation : vector3 = HexTransform.Translation+PivotVariation
            NewTransform:=transform{Translation:=Translation,Rotation:=HexaPropPosition.GetTransform().Rotation,Scale:=NormalScale}
            ResultTuple:=SpawnProp(HexaPet.PropAsset,NewTransform)
            if(ValidProp := ResultTuple(0)?):
                ##Print("Temporal Prop spawned")
                SelectedPetProp.Dispose()
                set SelectedPetProp = ValidProp
                HexaPet.ResetHealth()
                SetHealthUI(SelectedPetProp,HexaPet.MaxHealth,HexaPet.Health,HexaPet.IsPlayerPet)
                SetStar(SelectedPetProp,HexaPet.Tier)
                set MaybeSelectedPet = option{HexaPet}  
        spawn:
            Delay_NavigationCanvas()

    CleanAllPetProps() : void =
        for(HexaAIPet : AllHexaAIPets, Prop := HexaAIPet.MaybeProp?, Prop.IsValid[]):
            Prop.Dispose()
        for(HexaPlayerPet : AllHexaPlayerPets, Prop := HexaPlayerPet.MaybeProp?, Prop.IsValid[]):
            Prop.Dispose()
              
    StartFight(Message : widget_message) : void =
        if(not GameIsRunning?):
            set GameIsRunning = true
            spawn:
                StartHexGame()

    TryPlacePet(Message : widget_message) : void =
        var NumUsedPets : int = 0
        for(HexaPlayerPet : AllHexaPlayerPets):
            if(HexaPlayerPet.IsBeingUsed?):
                set NumUsedPets = NumUsedPets + 1
        NumPetsPlaced.SetText(BasicMessage("{NumUsedPets}/5 used slots"))
        UpdateNavigationCanvas()
        if(NumUsedPets<5):
            if(3>AgentFocus(0)):
                SelectedPetProp.Dispose()
                if(MaybeHexPet:=AllHexBoxes[AgentFocus(0)][AgentFocus(1)]):
                    if(ValidHexPet := MaybeHexPet?):
                        var SpaceFound : int = -1
                        var IndexA : int = 0
                        #Print("The position where is your trying to set is full")
                        loop:  
                            if(HexaBox := AllHexBoxes[AgentFocus(0)][AgentFocus(1)+IndexA], ValidHexPetA := HexaBox?):
                                #Print("Position {AgentFocus(1)+IndexA} is full")
                                set IndexA = IndexA + 1
                            else:
                                if(AgentFocus(1)+IndexA <> -1, AgentFocus(1)+IndexA <= AllHexBoxes[0].Length-1):
                                    set SpaceFound = AgentFocus(1)+IndexA
                                    #Print("Position {AgentFocus(1)+IndexA} is empty and is valid")
                                    break
                                else:
                                    break
                            if(IndexA>AllHexBoxes[0].Length-1):
                                break

                        if(SpaceFound<>-1):
                            #Print("Found Empty position of the right side. Moving pets")
                            for(Index:=0..AllHexBoxes[0].Length-1):
                                if:
                                    SpaceFound-Index-1 >= AgentFocus(1)
                                    OtherMaybeHexPet := AllHexBoxes[AgentFocus(0)][SpaceFound-Index-1] 
                                    set AllHexBoxes[AgentFocus(0)][SpaceFound-Index] = OtherMaybeHexPet
                                    set AllHexBoxes[AgentFocus(0)][SpaceFound-Index-1] = false
                                    ValidOtherHexPet := OtherMaybeHexPet?
                                then:
                                    #Print("Moved pet on position {SpaceFound-Index-1} to position {SpaceFound-Index}")
                                    ValidOtherHexPet.SetPosition(AgentFocus(0),SpaceFound-Index)

                            #Print("Placing new pet in the correct place {AgentFocus(1)}")
                            if(HexMiniGameSelectorUI := MaybeHexMiniGameSelectorUI?, HexaPet := AllHexaPlayerPets[HexMiniGameSelectorUI.CurrentOption]):
                                PlacePet(HexaPet,((AgentFocus(0),AgentFocus(1))))
                                
                        else:
                            var IndexB : int = 0
                            #Print("The position where is your trying to set is full")
                            loop:  
                                if(HexaBox := AllHexBoxes[AgentFocus(0)][AgentFocus(1)-IndexB], ValidHexPetA := HexaBox?):
                                    #Print("Position {AgentFocus(1)-IndexB} is full")
                                    set IndexB = IndexB + 1
                                else:
                                    if(AgentFocus(1)-IndexB <> -1, AgentFocus(1)-IndexB <= AllHexBoxes[0].Length-1):
                                        set SpaceFound = AgentFocus(1)-IndexB
                                        #Print("Position {AgentFocus(1)-IndexB} is empty and is valid")
                                        break
                                    else:
                                        break
                                if(IndexA>AllHexBoxes[0].Length-1):
                                    break
                            if(SpaceFound<>-1):
                                #Print("Found Empty position of the left side. Moving pets")
                                for(Index:=0..AllHexBoxes[0].Length-1):
                                    if:
                                        SpaceFound+Index+1 <= AgentFocus(1)
                                        OtherMaybeHexPet := AllHexBoxes[AgentFocus(0)][SpaceFound+Index+1] 
                                        set AllHexBoxes[AgentFocus(0)][SpaceFound+Index] = OtherMaybeHexPet
                                        set AllHexBoxes[AgentFocus(0)][SpaceFound+Index+1] = false
                                        ValidOtherHexPet := OtherMaybeHexPet?
                                    then:
                                        #Print("Moved pet on position {SpaceFound+Index+1} to position {SpaceFound+Index}")
                                        ValidOtherHexPet.SetPosition(AgentFocus(0),SpaceFound+Index)
                                #Print("Placing new pet in the correct place {AgentFocus(1)}")

                                if(HexMiniGameSelectorUI := MaybeHexMiniGameSelectorUI?, HexaPet := AllHexaPlayerPets[HexMiniGameSelectorUI.CurrentOption]):
                                    PlacePet(HexaPet,((AgentFocus(0),AgentFocus(1))))
                        if(SpaceFound=-1):
                            #Print("ERROR, SPACE NOT FOUND!")
                    else if(HexMiniGameSelectorUI := MaybeHexMiniGameSelectorUI?, HexaPet := AllHexaPlayerPets[HexMiniGameSelectorUI.CurrentOption]):
                        PlacePet(HexaPet,((AgentFocus(0),AgentFocus(1))))               
                spawn:
                    DelayChangeUI_Nav_Selector(Message)
                UpdateNavigationCanvas()
            else:
                spawn:
                    RunInvalidPositionWarning()
        else:
            spawn:
                RunMaxPetsReached()

    RunMaxPetsReached()<suspends> : void =
        ButtonNavigationOK.SetText(BasicMessage("LIMIT REACHED!"))
        Sleep(1.0)
        ButtonNavigationOK.SetText(BasicMessage("OK"))

    RunInvalidPositionWarning()<suspends> : void =
        ButtonNavigationOK.SetText(BasicMessage("INVALID!"))
        Sleep(1.0)
        ButtonNavigationOK.SetText(BasicMessage("OK"))

    DelayChangeUI_Nav_Selector(Message : widget_message)<suspends> : void =
        Remove_NavigationCanvas(Message)
        Sleep(0.1)
        Display_HexMiniGameSelectorUI()
        
        
    PlacePet(HexaPet : hexagame_pet, Position : tuple(int,int)) : void =
        #Print("Placing pet on {Position(0)}{Position(1)}")
        set HexaPet.IsBeingUsed = true
        HexaPet.SetPosition(Position(0),Position(1))
        if(set AllHexBoxes[AgentFocus(0)][AgentFocus(1)] = option{HexaPet}):
            HexaPet.InitHexaPet()
        if(BakeryMachinesManager := MaybeBakeryMachinesManager?, var AllPlayerPets : []pet_data_table = BakeryMachinesManager.PlayerDataCache.AllPets,var PetHexaUsed : pet_data_table = AllPlayerPets[HexaPet.AllPetsIndex]):
            set PetHexaUsed = BakeryMachinesManager.PlayersDataManager.SetHexaGamePosition(PetHexaUsed,Position(0),Position(1))
            if(set AllPlayerPets[HexaPet.AllPetsIndex] = PetHexaUsed):
                NewPlayerDataTable:=BakeryMachinesManager.PlayersDataManager.SavePets(BakeryMachinesManager.PlayerDataCache,AllPlayerPets)
                BakeryMachinesManager.SetNewPlayerDataCache(NewPlayerDataTable)

    WithdrawPet(HexaPet : hexagame_pet, Position : tuple(int,int)) : void =
        set HexaPet.IsBeingUsed = false
        HexaPet.SetPosition(-1,-1)
        if(ValidProp := HexaPet.MaybeProp?):
            ValidProp.Dispose()
        if(BakeryMachinesManager := MaybeBakeryMachinesManager?, var AllPlayerPets : []pet_data_table = BakeryMachinesManager.PlayerDataCache.AllPets,var PetHexaUsed : pet_data_table = AllPlayerPets[HexaPet.AllPetsIndex]):
            set PetHexaUsed = BakeryMachinesManager.PlayersDataManager.SetHexaGamePosition(PetHexaUsed,-1,-1)
            if(set AllPlayerPets[HexaPet.AllPetsIndex] = PetHexaUsed):
                NewPlayerDataTable:=BakeryMachinesManager.PlayersDataManager.SavePets(BakeryMachinesManager.PlayerDataCache,AllPlayerPets)
                BakeryMachinesManager.SetNewPlayerDataCache(NewPlayerDataTable)
        if(set AllHexBoxes[AgentFocus(0)][AgentFocus(1)] = false):

    Handler_Navigation_ToPetSelector(Message : widget_message) : void =
        Remove_NavigationCanvasWithoutMessage()
        spawn:
            Delay_ToPetSelector()

    Handler_ChallengeSelector_ToPetSelector(Message : widget_message) : void =
        Remove_ChallengeCanvasWithoutMessage()
        spawn:
            Delay_ToPetSelector()

    RefreshHexagameUI() : void =
        if(HexaMinigameManagerUI := MaybeHexMiniGameUI?):
            HexaMinigameManagerUI.RefreshAll()
        var AllAgentPets : []hexagame_pet = array{}
        var AllAIPets : []hexagame_pet = array{}

    Remove_NavigationCanvas(Message : widget_message) : void =
        SelectedPetProp.Dispose()
        set UIIsNavigationCanvas = false
        ##Print("Trying to remove UI")
        if:
            Agent := MaybeAgent?
            Player:=player[Agent]
            PlayerUI:=GetPlayerUI[Player]
            Canvas := MaybeNavigationCanvas?
        then:
            PlayerUI.RemoveWidget(Canvas)
            RemoveNavigationDelay_OnRemove()


    Remove_NavigationCanvasWithoutMessage() : void =
        set UIIsNavigationCanvas = false
        SelectedPetProp.Dispose()
        if:
            Agent := MaybeAgent?
            Player:=player[Agent]
            PlayerUI:=GetPlayerUI[Player]
            Canvas := MaybeNavigationCanvas?
        then:
            PlayerUI.RemoveWidget(Canvas)
            RemoveNavigationDelay_OnRemove()

    RemoveNavigationDelay_OnAdd() : void =
        ButtonChallengesLeft.SetText(BasicMessage("◀"))
        ButtonChallengesRight.SetText(BasicMessage("▶"))
        ButtonNavigationOK.SetText(BasicMessage("OK"))

    RemoveNavigationDelay_OnRemove() : void =
        ButtonChallengesLeft.SetText(BasicMessage("◀ "))
        ButtonChallengesRight.SetText(BasicMessage("▶ "))
        ButtonNavigationOK.SetText(BasicMessage("OK "))

    Remove_ChallengeCanvasAndCamera(Message : widget_message) : void =
        CleanAllPetProps()
        set UIIsChallengeCanvas = false
        if(BakeryMachinesManager:=MaybeBakeryMachinesManager?, Agent:=MaybeAgent?, Player:= player[Agent], Player.IsActive[]):
            BakeryMachinesManager.CheckMusicPlay(BAKERSMUSIC_SOUND)
            BakeryMachinesManager.GameManager.ChangeAgentClass_ToPickMode(Agent)
            #PrintWarning("On Remove_ChallengeCanvasAndCamera of hexa_minigame_manager to do calcule inventory limits")
            BakeryMachinesManager.RefreshAllUI()
        if:
            Agent := MaybeAgent?
            Player:=player[Agent]
            PlayerUI:=GetPlayerUI[Player]
            Canvas := MaybeChallengeCanvas?
            BakeryMachinesManager:=MaybeBakeryMachinesManager?
            MainInventoryUI:=BakeryMachinesManager.MaybeMainInventoryUI?
        then:
            spawn:
                BakeryMachinesManager.GameManager.VFXManager.StopSound(HEXAGAMEMUSIC_SOUND,0.0,option{Agent},false)
            PlayerUI.RemoveWidget(Canvas)
            RemoveChallengeDelay_OnRemove()
            CameraFixedPoint_HexaEnemies.RemoveFrom(Agent)
            CameraFixedPoint_Hexa.RemoveFrom(Agent)
            set IsUsingHexaGame = false
            MainInventoryUI.SimplifiedRefreshAll()
            BakeryMachinesManager.SetNewPlayerDataCache(PlayersDataManager.SetChallengePointer(BakeryMachinesManager.PlayerDataCache,ChallengeCurrentSelection))

    Remove_ChallengeCanvasWithoutMessage() : void =
        set UIIsChallengeCanvas = false
        if:
            Agent := MaybeAgent?
            Player:=player[Agent]
            PlayerUI:=GetPlayerUI[Player]
            Canvas := MaybeChallengeCanvas?
        then:
            PlayerUI.RemoveWidget(Canvas)
            RemoveChallengeDelay_OnRemove()
    

    RemoveChallengeDelay_OnAdd() : void =
        ButtonChallengesOK.SetText(BasicMessage("OK"))

    RemoveChallengeDelay_OnRemove() : void =
        ButtonChallengesOK.SetText(BasicMessage("OK "))

    Delay_ChallengeCanvas()<suspends> : void =
        Sleep(0.1)
        Display_ChallengeCanvas()

    Display_ChallengeCanvas() : void =
        if(not UIIsChallengeCanvas?, BakeryMachinesManager:=MaybeBakeryMachinesManager?):
            if:
                Agent := MaybeAgent?
                PlayerUI:=GetPlayerUI[player[Agent]]
                Canvas := MaybeChallengeCanvas?
                ActionInfoUI:=BakeryMachinesManager.MaybeActionInfoUI?
                MainInventoryUI:=BakeryMachinesManager.MaybeMainInventoryUI?
                not ActionInfoUI.IsDungeonStarting?
            then:
                BakeryMachinesManager.CheckMusicPlay(HEXAGAMEMUSIC_SOUND)
                RefreshChallengeUI()
                set UIIsChallengeCanvas = true
                var Index : int = 0 
                BakeryMachinesManager.GameManager.ChangeAgentClass_ToHexaGameMode(Agent)
                MainInventoryUI.Hide_All()
                CameraFixedPoint_HexaEnemies.AddTo(Agent)
                set AgentFocus = (0,0)
                PlayerUI.AddWidget(Canvas,player_ui_slot{ZOrder:=0, InputMode := ui_input_mode.All})
                RemoveChallengeDelay_OnAdd()
    
    Delay_NavigationCanvas()<suspends> : void =
        if(not UIIsNavigationCanvas?):
            set UIIsNavigationCanvas = true
            Sleep(0.1)
            var Index : int = 0 
            if:
                Agent := MaybeAgent?
                PlayerUI:=GetPlayerUI[player[Agent]]
                Canvas := MaybeNavigationCanvas?
            then:
                set AgentFocus = (0,0)
                PlayerUI.AddWidget(Canvas,player_ui_slot{ZOrder:=0, InputMode := ui_input_mode.All})
                RemoveNavigationDelay_OnAdd()
                UpdateNavigationCanvas()

    UpdateNavigationCanvas() : void =
        var NumUsedPets : int = 0
        for(HexaPlayerPet : AllHexaPlayerPets):
            if(HexaPlayerPet.IsBeingUsed?):
                set NumUsedPets = NumUsedPets + 1
        NumPetsPlaced.SetText(BasicMessage("{NumUsedPets}/5 used slots"))

    MakeNavigationCanvas(): canvas=
        Canvas := canvas:
            Slots:=array:
                #Main overlay
                canvas_slot:
                    Anchors:=anchors{Minimum:=vector2{X:=0.5,Y:=0.5},Maximum:=vector2{X:=0.5,Y:=0.5}}
                    ZOrder:=0
                    SizeToContent:=true
                    Offsets:=margin{Left:=0.0,Top:=0.0,Right:=0.0,Bottom:=0.0} #PositionX, PositionY, SizeX, SizeY
                    Alignment:=vector2{X:=0.5,Y:=0.5}
                    Widget:=overlay:
                        Slots:=array:
                            overlay_slot:
                                HorizontalAlignment:=horizontal_alignment.Center
                                VerticalAlignment:=vertical_alignment.Center
                                Widget:=stack_box:  
                                    Orientation:=orientation.Horizontal
                                        Slots:=array:
                                            stack_box_slot: 
                                                Padding:=margin{Right:=5.0}
                                                HorizontalAlignment:=horizontal_alignment.Fill
                                                VerticalAlignment:=vertical_alignment.Fill
                                                Widget:=overlay:
                                                    Slots:=array:
                                                        overlay_slot:
                                                            Widget:=texture_block{DefaultImage:=Images.EmptyTexture,DefaultDesiredSize:=vector2{X:=1.0,Y:=65.0}}
                                                        overlay_slot:
                                                            VerticalAlignment:=vertical_alignment.Fill
                                                            HorizontalAlignment:=horizontal_alignment.Fill
                                                            Widget:=ButtonNavigationLeft    
                                            stack_box_slot: 
                                                HorizontalAlignment:=horizontal_alignment.Fill
                                                VerticalAlignment:=vertical_alignment.Fill
                                                Widget:=texture_block{DefaultImage:=Images.EmptyTexture,DefaultDesiredSize:=vector2{X:=100.0,Y:=56.0}}                             
                                            stack_box_slot: 
                                                HorizontalAlignment:=horizontal_alignment.Fill
                                                Padding:=margin{Left:=5.0}
                                                Widget:=overlay:
                                                    Slots:=array:
                                                        overlay_slot:
                                                            Widget:=texture_block{DefaultImage:=Images.EmptyTexture,DefaultDesiredSize:=vector2{X:=65.0,Y:=65.0}}
                                                        overlay_slot:
                                                            HorizontalAlignment:=horizontal_alignment.Fill
                                                            VerticalAlignment:=vertical_alignment.Fill
                                                            Widget:=ButtonNavigationRight
                                                            
                            overlay_slot:
                                HorizontalAlignment:=horizontal_alignment.Center
                                VerticalAlignment:=vertical_alignment.Center
                                Widget:= stack_box:
                                    Orientation:=orientation.Vertical
                                    Slots:=array:
                                        stack_box_slot: 
                                            HorizontalAlignment:=horizontal_alignment.Center
                                            VerticalAlignment:=vertical_alignment.Center
                                            Padding:=margin{Top:=5.0,Bottom:=5.0}
                                            Widget:=overlay:
                                                Slots:=array:
                                                    <#
                                                    overlay_slot:
                                                        Widget:=texture_block{DefaultImage:=Images.Img_LineBackground,DefaultDesiredSize:=vector2{X:=370.0,Y:=10.0}}#>
                                                    overlay_slot:
                                                        HorizontalAlignment:=horizontal_alignment.Fill
                                                        VerticalAlignment:=vertical_alignment.Fill
                                                        Widget:=overlay:
                                                            Slots:=array:
                                                                overlay_slot:
                                                                    Widget:=texture_block{DefaultImage:=Images.EmptyTexture,DefaultDesiredSize:=vector2{X:=65.0,Y:=65.0}}
                                                                overlay_slot:
                                                                    VerticalAlignment:=vertical_alignment.Fill
                                                                    HorizontalAlignment:=horizontal_alignment.Fill
                                                                    Widget:=ButtonNavigationTop
                                        stack_box_slot: 
                                            HorizontalAlignment:=horizontal_alignment.Center
                                            VerticalAlignment:=vertical_alignment.Center
                                            Padding:=margin{Top:=5.0,Bottom:=5.0}
                                            Widget:=overlay:
                                                Slots:=array:
                                                    overlay_slot:
                                                        HorizontalAlignment:=horizontal_alignment.Fill
                                                        VerticalAlignment:=vertical_alignment.Fill
                                                        Widget:=overlay:
                                                            Slots:=array:
                                                                overlay_slot:
                                                                    Widget:=texture_block{DefaultImage:=Images.EmptyTexture,DefaultDesiredSize:=vector2{X:=65.0,Y:=65.0}}
                                                                overlay_slot:
                                                                    VerticalAlignment:=vertical_alignment.Fill
                                                                    HorizontalAlignment:=horizontal_alignment.Fill
                                                                    Widget:=ButtonNavigationOK
                                        stack_box_slot: 
                                            HorizontalAlignment:=horizontal_alignment.Center
                                            VerticalAlignment:=vertical_alignment.Center
                                            Padding:=margin{Top:=5.0,Bottom:=5.0}
                                            Widget:=overlay:
                                                Slots:=array:
                                                    <#overlay_slot:
                                                        Widget:=texture_block{DefaultImage:=Images.Img_LineBackground,DefaultDesiredSize:=vector2{X:=370.0,Y:=10.0}}#>
                                                    overlay_slot:
                                                        HorizontalAlignment:=horizontal_alignment.Fill
                                                        VerticalAlignment:=vertical_alignment.Fill
                                                        Widget:=overlay:
                                                            Slots:=array:
                                                                overlay_slot:
                                                                    Widget:=texture_block{DefaultImage:=Images.EmptyTexture,DefaultDesiredSize:=vector2{X:=65.0,Y:=65.0}}
                                                                overlay_slot:
                                                                    VerticalAlignment:=vertical_alignment.Fill
                                                                    HorizontalAlignment:=horizontal_alignment.Fill
                                                                     Widget:=ButtonNavigationBottom

                #Cancel button
                canvas_slot:
                    Anchors:=anchors{Minimum:=vector2{X:=0.5,Y:=0.0},Maximum:=vector2{X:=0.5,Y:=0.0}}
                    ZOrder:=0
                    SizeToContent:=true
                    Offsets:=margin{Left:=0.0,Top:=15.0,Right:=0.0,Bottom:=0.0} #PositionX, PositionY, SizeX, SizeY
                    Alignment:=vector2{X:=0.5,Y:=0.0}
                    Widget:=overlay:
                        Slots:=array:
                            overlay_slot:
                                Widget:=texture_block{DefaultImage:=Images.EmptyTexture,DefaultDesiredSize:=vector2{X:=1.0,Y:=80.0}}
                            overlay_slot:
                                HorizontalAlignment:=horizontal_alignment.Center
                                VerticalAlignment:=vertical_alignment.Fill
                                Widget:=ButtonNavigationCancel
                #Num pets placed
                canvas_slot:
                    Anchors:=anchors{Minimum:=vector2{X:=0.5,Y:=1.0},Maximum:=vector2{X:=0.5,Y:=1.0}}
                    ZOrder:=0
                    SizeToContent:=true
                    Offsets:=margin{Left:=0.0,Top:=0.0,Right:=0.0,Bottom:=15.0} #PositionX, PositionY, SizeX, SizeY
                    Alignment:=vector2{X:=0.5,Y:=0.0}
                    Widget:=overlay:
                        Slots:=array:
                            overlay_slot:
                                HorizontalAlignment:=horizontal_alignment.Center
                                VerticalAlignment:=vertical_alignment.Center
                                Widget:=NumPetsPlaced                 

    RefreshChallengeUI() : void =

        for(HexaPet : AllHexaAIPets):
            HexaPet.SetIsBeingUsed(false)
            HexaPet.DespawnPetProp()

        if(ChallengeCurrentSelection>HexaChallengesDatabase.AllChallenges.Length-1):
            set ChallengeCurrentSelection = HexaChallengesDatabase.AllChallenges.Length-1

        if(BakeryManager := MaybeBakeryMachinesManager?):
            set AllHexaAIPets=HexaChallengesDatabase.GetChallengeGroup(AllHexaAIPets,Self,BakeryManager.GameManager,ChallengeCurrentSelection)
            ChallengesCompleted:=BakeryManager.PlayerDataCache.HexaGame_AllChallengesCompleted.Length
            NumChallengesCompleted.SetText(BasicMessage("Challenges completed {ChallengesCompleted}/{HexaChallengesDatabase.AllChallenges.Length}"))

        #for(Index:=0..AllHexaAIPets.Length-1, Pet := AllHexaAIPets[Index], Pet.IsBeingUsed?):
            ##Print("Pet Type: {Pet.PetType} is being used on position [{Pet.Position(0)}][{Pet.Position(1)}]")

            if(Agent := MaybeAgent?, Player := player[Agent], Player.IsActive[]):
                for(Index:=0..AllSlotChallenge_Images.Length-1):
                    if:
                        Image := AllSlotChallenge_Images[Index] 
                        Description := AllSlotChallenge_Description[Index]
                        RewardsInfo := AllSlotChallenge_IsCompleted[Index]
                    then:
                        ChallengeBoxPosition:=ChallengeCurrentSelection-2+Index
                        TupleChallengeName:=HexaChallengesDatabase.GetChallengeName(ChallengeBoxPosition)

                        if(not TupleChallengeName(1)?):
                            #Image.SetTint(White)
                            Image.SetImage(Images.EmptyTexture)
                            Image.SetDesiredSize(vector2{X:=75.0,Y:=75.0})
                            RewardsInfo.SetTextColor(White)
                            RewardsInfo.SetText(BasicMessage("-"))
                            Description.SetText(BasicMessage("-"))
                        else:
                            var IsChallengeAlreadyCompleted : logic = false
                            var IsFirstTime : logic = true
                            Description.SetTextColor(UpgradeTextColor)
                            for(ChallengeAlreadyCompleted : BakeryManager.PlayerDataCache.HexaGame_AllChallengesCompleted, ChallengeAlreadyCompleted = ChallengeBoxPosition):
                                set IsChallengeAlreadyCompleted = true
                                set IsFirstTime = false
                                Description.SetTextColor(CurrentTextColor)

                            GatherCalculation:=CalculateGatherReward(ChallengeBoxPosition, IsFirstTime,true,1)
                            AllCollectibleHexaChallenges:=BakeryManager.GameManager.CollectiblesManager.AllCollectibleHexaChallenges
                            NumChallengesUnlocked:=CalculeNumChallengesUnlocked(BakeryManager.PlayerDataCache.AllHexaGame_TotemsUnlocked,BakeryManager.HexaGameManager.HexaChallengesDatabase.AllChallenges.Length,AllCollectibleHexaChallenges.Length)
                            if(ChallengeBoxPosition<=NumChallengesUnlocked):
                                Image.SetImage(Images.AgedSaphire_Icon)
                                Image.SetDesiredSize(vector2{X:=75.0,Y:=75.0})
                                Description.SetText(BasicMessage(TupleChallengeName(0)))
                                if(IsChallengeAlreadyCompleted?):
                                    Image.SetImage(CompletedImage)
                                else:
                                    Image.SetImage(NotCompletedImage)
                                Image.SetTint(White)
                            else:
                                Image.SetImage(Images.Cross_Icon)
                                Image.SetTint(Red)
                                Image.SetDesiredSize(vector2{X:=75.0,Y:=75.0})
                                Description.SetTextColor(Red)
                                Description.SetText(BasicMessage("LOCKED"))
                            RewardsInfo.SetTextColor(GatherInfoColor)
                            RewardsInfo.SetText(BasicMessage("x{GetOptimizedMagnitudeLetter(GatherCalculation)} Gather +{CalculeHBLevel(ChallengeBoxPosition)*5}%"))

                            if(ChallengeBoxPosition = ChallengeCurrentSelection):
                                NumHits:=CalculateGatherReward(ChallengeCurrentSelection,IsFirstTime,true,1)
                                CoalReward:=CalculeCoalGatherReward(ChallengeCurrentSelection, IsFirstTime, true)
                                MutantReward:=CalculeMutantMaterialGatherReward(ChallengeCurrentSelection, IsFirstTime, true, NumHits)
                                CoalReward_Text.SetText(BasicMessage("x {GetOptimizedMagnitudeLetter(CoalReward)}"))
                                MutantMaterialReward_Text.SetText(BasicMessage("x {GetOptimizedMagnitudeLetter(MutantReward)}"))

                    else:
                        PrintFatalError("Inavalid  slot {Index}")
                     

        for(Index:=0..AllTextureBlock_HealthLines_AI.Length-1):
            if:
                TextureBlock_HealthLine_AI:=AllTextureBlock_HealthLines_AI[Index]
                TextureBlock_PetImages_AI:=AllTextureBlock_PetImages_AI[Index]
                TextBlock_Classes_AI:=AllTextBlock_Classes_AI[Index]
                TextBlock_PackA_AI:=AllTextBlock_PackA_AI[Index]
                TextBlock_PackB_AI:=AllTextBlock_PackB_AI[Index]
                TextureBlock_Purity_AI:=AllTextBlock_Purity_AI[Index]
                TextBlock_Health_AI:=AllTextBlock_Health_AI[Index]
            then:
                TextureBlock_HealthLine_AI.SetDesiredSize(vector2{X:=MaxHealthSize_PetHexaUI})
                TextureBlock_PetImages_AI.SetImage(Images.Cross_Icon)
                TextureBlock_PetImages_AI.SetDesiredSize(vector2{X:=100.0,Y:=100.0})
                TextureBlock_Purity_AI.SetTextColor(White)
                TextureBlock_Purity_AI.SetText(BasicMessage("-"))
                TextBlock_PackA_AI.SetText(BasicMessage("-"))
                TextBlock_PackA_AI.SetTextColor(White)
                TextBlock_PackB_AI.SetText(BasicMessage("-"))
                TextBlock_PackB_AI.SetTextColor(White)
                TextBlock_Classes_AI.SetText(BasicMessage("-"))
                TextBlock_Classes_AI.SetTextColor(White)
                TextBlock_Health_AI.SetText(BasicMessage("-"))
                TextBlock_Health_AI.SetTextColor(White)
        
        for(Index:=0..AllHexaAIPets.Length-1):
            if:
                TextureBlock_HealthLine_AI:=AllTextureBlock_HealthLines_AI[Index]
                TextureBlock_PetImages_AI:=AllTextureBlock_PetImages_AI[Index]
                TextBlock_Classes_AI:=AllTextBlock_Classes_AI[Index]
                TextBlock_PackA_AI:=AllTextBlock_PackA_AI[Index]
                TextBlock_PackB_AI:=AllTextBlock_PackB_AI[Index]
                TextureBlock_Purity_AI:=AllTextBlock_Purity_AI[Index]
                TextBlock_Health_AI:=AllTextBlock_Health_AI[Index]
            then:
                if(HexaPet := AllHexaAIPets[Index], HexaPet.IsBeingUsed?):
                    HexaPet.SetIsBeingUsed(true)
                    SetPetClass(HexaPet, TextBlock_Classes_AI, TextBlock_PackA_AI, TextBlock_PackB_AI)
                    TextureBlock_HealthLine_AI.SetTint(Red)
                    SetPetHealthBar(HexaPet,HexaPet.Health,TextureBlock_HealthLine_AI, TextBlock_Health_AI)
                    SetPetImage(HexaPet,TextureBlock_PetImages_AI)
                    SetPetPurity(HexaPet,HexaPet.Tier,TextureBlock_Purity_AI)
                    HexaPet.SpawnPetProp()
                else:
                    TextureBlock_HealthLine_AI.SetTint(White)

    NavigateChallenges5Left(Message : widget_message) : void =
        ##Print("Navigating left")
        set ChallengeCurrentSelection = ChallengeCurrentSelection - 5
        if(ChallengeCurrentSelection < 0):
            set ChallengeCurrentSelection = 0
        RefreshChallengeUI()

    NavigateChallenges5Right(Message : widget_message) : void =
        ##Print("Navigating right")
        set ChallengeCurrentSelection = ChallengeCurrentSelection + 5
        if(BakeryManager := MaybeBakeryMachinesManager?, Message.Player.IsActive[]):
            AllCollectibleHexaChallenges:=BakeryManager.GameManager.CollectiblesManager.AllCollectibleHexaChallenges
            NumChallengesUnlocked:=CalculeNumChallengesUnlocked(BakeryManager.PlayerDataCache.AllHexaGame_TotemsUnlocked,BakeryManager.HexaGameManager.HexaChallengesDatabase.AllChallenges.Length,AllCollectibleHexaChallenges.Length)
            if(NumChallengesUnlocked<ChallengeCurrentSelection):
                set ChallengeCurrentSelection = NumChallengesUnlocked
        RefreshChallengeUI()

    NavigateChallengesLeft(Message : widget_message) : void =
        ##Print("Navigating left")
        set ChallengeCurrentSelection = ChallengeCurrentSelection - 1
        if(ChallengeCurrentSelection < 0):
            set ChallengeCurrentSelection = 0
        RefreshChallengeUI()

    NavigateChallengesRight(Message : widget_message) : void =
        ##Print("Navigating right")
        set ChallengeCurrentSelection = ChallengeCurrentSelection + 1 
        if(BakeryManager := MaybeBakeryMachinesManager?, Message.Player.IsActive[]):
            AllCollectibleHexaChallenges:=BakeryManager.GameManager.CollectiblesManager.AllCollectibleHexaChallenges
            NumChallengesUnlocked:=CalculeNumChallengesUnlocked(BakeryManager.PlayerDataCache.AllHexaGame_TotemsUnlocked,BakeryManager.HexaGameManager.HexaChallengesDatabase.AllChallenges.Length,AllCollectibleHexaChallenges.Length)
            if(NumChallengesUnlocked<ChallengeCurrentSelection):
                set ChallengeCurrentSelection = NumChallengesUnlocked
        RefreshChallengeUI()
    
    MakeChallengeCanvas(): canvas=
        Canvas := canvas:
            Slots:=array:
                #Main overlay
                canvas_slot:
                    Anchors:=anchors{Minimum:=vector2{X:=1.0,Y:=0.0},Maximum:=vector2{X:=1.0,Y:=0.0}}
                    ZOrder:=0
                    SizeToContent:=true
                    Offsets:=margin{Left:=-5.0,Top:=5.0,Right:=0.0,Bottom:=0.0} #PositionX, PositionY, SizeX, SizeY
                    Alignment:=vector2{X:=1.0,Y:=0.0}
                    Widget:=overlay:
                        Slots:=array:
                            overlay_slot:
                                HorizontalAlignment:=horizontal_alignment.Fill
                                VerticalAlignment:=vertical_alignment.Fill
                                Widget:=stack_box:
                                    Orientation:=orientation.Vertical
                                        Slots:=array:
                                            stack_box_slot: #Slot 1
                                                Padding:=margin{Top:=ResourcersRowTopPadding,Bottom:=ResourcersRowBottomPadding}
                                                HorizontalAlignment:=horizontal_alignment.Fill
                                                VerticalAlignment:=vertical_alignment.Fill
                                                Widget:=overlay:
                                                    Slots:=array:
                                                        overlay_slot:
                                                            HorizontalAlignment:=horizontal_alignment.Fill
                                                            VerticalAlignment:=vertical_alignment.Fill
                                                            Widget:=EmptyImage_AI_Slot1
                                                        overlay_slot:
                                                            HorizontalAlignment:=horizontal_alignment.Fill
                                                            VerticalAlignment:=vertical_alignment.Fill
                                                            Widget:= texture_block{DefaultImage:=Images.Img_LineBackground}
                                                        overlay_slot:
                                                            HorizontalAlignment:=horizontal_alignment.Fill
                                                            VerticalAlignment:=vertical_alignment.Fill
                                                            Padding:=margin{Left:=PetsLineBoxPadding,Top:=PetsLineBoxPadding,Right:=PetsLineBoxPadding,Bottom:=PetsLineBoxPadding}
                                                            Widget:= texture_block{DefaultImage:=Images.Img_LineBackground}
                                                        overlay_slot:
                                                            HorizontalAlignment:=horizontal_alignment.Left
                                                            VerticalAlignment:=vertical_alignment.Fill
                                                            Padding:=margin{Left:=PetsLineBoxPadding,Top:=PetsLineBoxPadding,Right:=PetsLineBoxPadding,Bottom:=PetsLineBoxPadding}
                                                            Widget:=HealthLine_AI_Slot1
                                                        overlay_slot:
                                                            HorizontalAlignment:=horizontal_alignment.Left
                                                            VerticalAlignment:=vertical_alignment.Center
                                                            Widget:=overlay:
                                                                Slots:=array:
                                                                    overlay_slot:
                                                                        HorizontalAlignment:=horizontal_alignment.Center
                                                                        VerticalAlignment:=vertical_alignment.Center
                                                                        Padding:=margin{Left:=7.0,Top:=PetsLineBoxPadding,Right:=PetsLineBoxPadding,Bottom:=PetsLineBoxPadding}
                                                                        Widget:=PetImage_AI_Slot1
                                                        overlay_slot:
                                                            HorizontalAlignment:=horizontal_alignment.Left
                                                            VerticalAlignment:=vertical_alignment.Center
                                                            Padding:=margin{Left:=TextPetsBoxPadding}
                                                            Widget:=stack_box:
                                                                Orientation:=orientation.Vertical
                                                                    Slots:=array:
                                                                        
                                                                        stack_box_slot: #Class
                                                                            HorizontalAlignment:=horizontal_alignment.Left
                                                                            VerticalAlignment:=vertical_alignment.Center
                                                                            Padding:=margin{Left:=TextPetsBoxPadding}
                                                                            Widget:=stack_box:
                                                                                Orientation:=orientation.Horizontal
                                                                                Slots:=array:
                                                                                    stack_box_slot: 
                                                                                        Padding:=margin{Left:=2.5,Right:=2.5}
                                                                                        Widget:=Class_AI_Slot1
                                                                                    stack_box_slot: 
                                                                                        Padding:=margin{Left:=2.5, Right:=1.0}
                                                                                        Widget:=PackA_AI_Box1
                                                                                    stack_box_slot: 
                                                                                        Padding:=margin{Right:=2.5}
                                                                                        Widget:=PackB_AI_Box1
                                                                        stack_box_slot: #Purity
                                                                            HorizontalAlignment:=horizontal_alignment.Left
                                                                            VerticalAlignment:=vertical_alignment.Center
                                                                            Padding:=margin{Left:=TextPetsBoxPadding}
                                                                            Widget:=Purity_AI_Slot1
                                                                        stack_box_slot: #Health
                                                                            HorizontalAlignment:=horizontal_alignment.Left
                                                                            VerticalAlignment:=vertical_alignment.Center
                                                                            Padding:=margin{Left:=TextPetsBoxPadding}
                                                                            Widget:=Health_AI_Slot1
                                            stack_box_slot: #Slot 2
                                                Padding:=margin{Top:=ResourcersRowTopPadding,Bottom:=ResourcersRowBottomPadding}
                                                HorizontalAlignment:=horizontal_alignment.Fill
                                                VerticalAlignment:=vertical_alignment.Fill
                                                Widget:=overlay:
                                                    Slots:=array:
                                                        overlay_slot:
                                                            HorizontalAlignment:=horizontal_alignment.Fill
                                                            VerticalAlignment:=vertical_alignment.Fill
                                                            Widget:=EmptyImage_AI_Slot2
                                                        overlay_slot:
                                                            HorizontalAlignment:=horizontal_alignment.Fill
                                                            VerticalAlignment:=vertical_alignment.Fill
                                                            Widget:= texture_block{DefaultImage:=Images.Img_LineBackground}
                                                        overlay_slot:
                                                            HorizontalAlignment:=horizontal_alignment.Fill
                                                            VerticalAlignment:=vertical_alignment.Fill
                                                            Padding:=margin{Left:=PetsLineBoxPadding,Top:=PetsLineBoxPadding,Right:=PetsLineBoxPadding,Bottom:=PetsLineBoxPadding}
                                                            Widget:= texture_block{DefaultImage:=Images.Img_LineBackground}
                                                        overlay_slot:
                                                            HorizontalAlignment:=horizontal_alignment.Left
                                                            VerticalAlignment:=vertical_alignment.Fill
                                                            Padding:=margin{Left:=PetsLineBoxPadding,Top:=PetsLineBoxPadding,Right:=PetsLineBoxPadding,Bottom:=PetsLineBoxPadding}
                                                            Widget:=HealthLine_AI_Slot2
                                                        overlay_slot:
                                                            HorizontalAlignment:=horizontal_alignment.Left
                                                            VerticalAlignment:=vertical_alignment.Center
                                                            Widget:=overlay:
                                                                Slots:=array:
                                                                    overlay_slot:
                                                                        HorizontalAlignment:=horizontal_alignment.Center
                                                                        VerticalAlignment:=vertical_alignment.Center
                                                                        Padding:=margin{Left:=7.0,Top:=PetsLineBoxPadding,Right:=PetsLineBoxPadding,Bottom:=PetsLineBoxPadding}
                                                                        Widget:=PetImage_AI_Slot2
                                                        overlay_slot:
                                                            HorizontalAlignment:=horizontal_alignment.Left
                                                            VerticalAlignment:=vertical_alignment.Center
                                                            Padding:=margin{Left:=TextPetsBoxPadding}
                                                            Widget:=stack_box:
                                                                Orientation:=orientation.Vertical
                                                                    Slots:=array:
                                                                        stack_box_slot: #Class
                                                                            HorizontalAlignment:=horizontal_alignment.Left
                                                                            VerticalAlignment:=vertical_alignment.Center
                                                                            Padding:=margin{Left:=TextPetsBoxPadding}
                                                                            Widget:=stack_box:
                                                                                Orientation:=orientation.Horizontal
                                                                                Slots:=array:
                                                                                    stack_box_slot: 
                                                                                        Padding:=margin{Left:=2.5,Right:=2.5}
                                                                                        Widget:=Class_AI_Slot2
                                                                                    stack_box_slot: 
                                                                                        Padding:=margin{Left:=2.5, Right:=1.0}
                                                                                        Widget:=PackA_AI_Box2
                                                                                    stack_box_slot: 
                                                                                        Padding:=margin{Right:=2.5}
                                                                                        Widget:=PackB_AI_Box2
                                                                        stack_box_slot: #Purity
                                                                            HorizontalAlignment:=horizontal_alignment.Left
                                                                            VerticalAlignment:=vertical_alignment.Center
                                                                            Padding:=margin{Left:=TextPetsBoxPadding}
                                                                            Widget:=Purity_AI_Slot2
                                                                        stack_box_slot: #Health
                                                                            HorizontalAlignment:=horizontal_alignment.Left
                                                                            VerticalAlignment:=vertical_alignment.Center
                                                                            Padding:=margin{Left:=TextPetsBoxPadding}
                                                                            Widget:=Health_AI_Slot2
                                            stack_box_slot: #Slot 3
                                                Padding:=margin{Top:=ResourcersRowTopPadding,Bottom:=ResourcersRowBottomPadding}
                                                HorizontalAlignment:=horizontal_alignment.Fill
                                                VerticalAlignment:=vertical_alignment.Fill
                                                Widget:=overlay:
                                                    Slots:=array:
                                                        overlay_slot:
                                                            HorizontalAlignment:=horizontal_alignment.Fill
                                                            VerticalAlignment:=vertical_alignment.Fill
                                                            Widget:=EmptyImage_AI_Slot3
                                                        overlay_slot:
                                                            HorizontalAlignment:=horizontal_alignment.Fill
                                                            VerticalAlignment:=vertical_alignment.Fill
                                                            Widget:= texture_block{DefaultImage:=Images.Img_LineBackground}
                                                        overlay_slot:
                                                            HorizontalAlignment:=horizontal_alignment.Fill
                                                            VerticalAlignment:=vertical_alignment.Fill
                                                            Padding:=margin{Left:=PetsLineBoxPadding,Top:=PetsLineBoxPadding,Right:=PetsLineBoxPadding,Bottom:=PetsLineBoxPadding}
                                                            Widget:= texture_block{DefaultImage:=Images.Img_LineBackground}
                                                        overlay_slot:
                                                            HorizontalAlignment:=horizontal_alignment.Left
                                                            VerticalAlignment:=vertical_alignment.Fill
                                                            Padding:=margin{Left:=PetsLineBoxPadding,Top:=PetsLineBoxPadding,Right:=PetsLineBoxPadding,Bottom:=PetsLineBoxPadding}
                                                            Widget:=HealthLine_AI_Slot3
                                                        overlay_slot:
                                                            HorizontalAlignment:=horizontal_alignment.Left
                                                            VerticalAlignment:=vertical_alignment.Center
                                                            Widget:=overlay:
                                                                Slots:=array:
                                                                    overlay_slot:
                                                                        HorizontalAlignment:=horizontal_alignment.Center
                                                                        VerticalAlignment:=vertical_alignment.Center
                                                                        Padding:=margin{Left:=7.0,Top:=PetsLineBoxPadding,Right:=PetsLineBoxPadding,Bottom:=PetsLineBoxPadding}
                                                                        Widget:=PetImage_AI_Slot3
                                                        overlay_slot:
                                                            HorizontalAlignment:=horizontal_alignment.Left
                                                            VerticalAlignment:=vertical_alignment.Center
                                                            Padding:=margin{Left:=TextPetsBoxPadding}
                                                            Widget:=stack_box:
                                                                Orientation:=orientation.Vertical
                                                                    Slots:=array:
                                                                        stack_box_slot: #Class
                                                                            HorizontalAlignment:=horizontal_alignment.Left
                                                                            VerticalAlignment:=vertical_alignment.Center
                                                                            Padding:=margin{Left:=TextPetsBoxPadding}
                                                                            Widget:=stack_box:
                                                                                Orientation:=orientation.Horizontal
                                                                                Slots:=array:
                                                                                    stack_box_slot: 
                                                                                        Padding:=margin{Left:=2.5,Right:=2.5}
                                                                                        Widget:=Class_AI_Slot3
                                                                                    stack_box_slot: 
                                                                                        Padding:=margin{Left:=2.5, Right:=1.0}
                                                                                        Widget:=PackA_AI_Box3
                                                                                    stack_box_slot: 
                                                                                        Padding:=margin{Right:=2.5}
                                                                                        Widget:=PackB_AI_Box3
                                                                        stack_box_slot: #Purity
                                                                            HorizontalAlignment:=horizontal_alignment.Left
                                                                            VerticalAlignment:=vertical_alignment.Center
                                                                            Padding:=margin{Left:=TextPetsBoxPadding}
                                                                            Widget:=Purity_AI_Slot3
                                                                        stack_box_slot: #Health
                                                                            HorizontalAlignment:=horizontal_alignment.Left
                                                                            VerticalAlignment:=vertical_alignment.Center
                                                                            Padding:=margin{Left:=TextPetsBoxPadding}
                                                                            Widget:=Health_AI_Slot3
                                            stack_box_slot: #Slot 4
                                                Padding:=margin{Top:=ResourcersRowTopPadding,Bottom:=ResourcersRowBottomPadding}
                                                HorizontalAlignment:=horizontal_alignment.Fill
                                                VerticalAlignment:=vertical_alignment.Fill
                                                Widget:=overlay:
                                                    Slots:=array:
                                                        overlay_slot:
                                                            HorizontalAlignment:=horizontal_alignment.Fill
                                                            VerticalAlignment:=vertical_alignment.Fill
                                                            Widget:=EmptyImage_AI_Slot4
                                                        overlay_slot:
                                                            HorizontalAlignment:=horizontal_alignment.Fill
                                                            VerticalAlignment:=vertical_alignment.Fill
                                                            Widget:= texture_block{DefaultImage:=Images.Img_LineBackground}
                                                        overlay_slot:
                                                            HorizontalAlignment:=horizontal_alignment.Fill
                                                            VerticalAlignment:=vertical_alignment.Fill
                                                            Padding:=margin{Left:=PetsLineBoxPadding,Top:=PetsLineBoxPadding,Right:=PetsLineBoxPadding,Bottom:=PetsLineBoxPadding}
                                                            Widget:= texture_block{DefaultImage:=Images.Img_LineBackground}
                                                        overlay_slot:
                                                            HorizontalAlignment:=horizontal_alignment.Left
                                                            VerticalAlignment:=vertical_alignment.Fill
                                                            Padding:=margin{Left:=PetsLineBoxPadding,Top:=PetsLineBoxPadding,Right:=PetsLineBoxPadding,Bottom:=PetsLineBoxPadding}
                                                            Widget:=HealthLine_AI_Slot4
                                                        overlay_slot:
                                                            HorizontalAlignment:=horizontal_alignment.Left
                                                            VerticalAlignment:=vertical_alignment.Center
                                                            Widget:=overlay:
                                                                Slots:=array:
                                                                    overlay_slot:
                                                                        HorizontalAlignment:=horizontal_alignment.Center
                                                                        VerticalAlignment:=vertical_alignment.Center
                                                                        Padding:=margin{Left:=7.0,Top:=PetsLineBoxPadding,Right:=PetsLineBoxPadding,Bottom:=PetsLineBoxPadding}
                                                                        Widget:=PetImage_AI_Slot4
                                                        overlay_slot:
                                                            HorizontalAlignment:=horizontal_alignment.Left
                                                            VerticalAlignment:=vertical_alignment.Center
                                                            Padding:=margin{Left:=TextPetsBoxPadding}
                                                            Widget:=stack_box:
                                                                Orientation:=orientation.Vertical
                                                                    Slots:=array:
                                                                        stack_box_slot: #Class
                                                                            HorizontalAlignment:=horizontal_alignment.Left
                                                                            VerticalAlignment:=vertical_alignment.Center
                                                                            Padding:=margin{Left:=TextPetsBoxPadding}
                                                                            Widget:=stack_box:
                                                                                Orientation:=orientation.Horizontal
                                                                                Slots:=array:
                                                                                    stack_box_slot: 
                                                                                        Padding:=margin{Left:=2.5,Right:=2.5}
                                                                                        Widget:=Class_AI_Slot4
                                                                                    stack_box_slot: 
                                                                                        Padding:=margin{Left:=2.5, Right:=1.0}
                                                                                        Widget:=PackA_AI_Box4
                                                                                    stack_box_slot: 
                                                                                        Padding:=margin{Right:=2.5}
                                                                                        Widget:=PackB_AI_Box4
                                                                        stack_box_slot: #Purity
                                                                            HorizontalAlignment:=horizontal_alignment.Left
                                                                            VerticalAlignment:=vertical_alignment.Center
                                                                            Padding:=margin{Left:=TextPetsBoxPadding}
                                                                            Widget:=Purity_AI_Slot4
                                                                        stack_box_slot: #Health
                                                                            HorizontalAlignment:=horizontal_alignment.Left
                                                                            VerticalAlignment:=vertical_alignment.Center
                                                                            Padding:=margin{Left:=TextPetsBoxPadding}
                                                                            Widget:=Health_AI_Slot4
                                            stack_box_slot: #Slot 5
                                                Padding:=margin{Top:=ResourcersRowTopPadding,Bottom:=ResourcersRowBottomPadding}
                                                HorizontalAlignment:=horizontal_alignment.Fill
                                                VerticalAlignment:=vertical_alignment.Fill
                                                Widget:=overlay:
                                                    Slots:=array:
                                                        overlay_slot:
                                                            HorizontalAlignment:=horizontal_alignment.Fill
                                                            VerticalAlignment:=vertical_alignment.Fill
                                                            Widget:=EmptyImage_AI_Slot5
                                                        overlay_slot:
                                                            HorizontalAlignment:=horizontal_alignment.Fill
                                                            VerticalAlignment:=vertical_alignment.Fill
                                                            Widget:= texture_block{DefaultImage:=Images.Img_LineBackground}
                                                        overlay_slot:
                                                            HorizontalAlignment:=horizontal_alignment.Fill
                                                            VerticalAlignment:=vertical_alignment.Fill
                                                            Padding:=margin{Left:=PetsLineBoxPadding,Top:=PetsLineBoxPadding,Right:=PetsLineBoxPadding,Bottom:=PetsLineBoxPadding}
                                                            Widget:= texture_block{DefaultImage:=Images.Img_LineBackground}
                                                        overlay_slot:
                                                            HorizontalAlignment:=horizontal_alignment.Left
                                                            VerticalAlignment:=vertical_alignment.Fill
                                                            Padding:=margin{Left:=PetsLineBoxPadding,Top:=PetsLineBoxPadding,Right:=PetsLineBoxPadding,Bottom:=PetsLineBoxPadding}
                                                            Widget:=HealthLine_AI_Slot5
                                                        overlay_slot:
                                                            HorizontalAlignment:=horizontal_alignment.Left
                                                            VerticalAlignment:=vertical_alignment.Center
                                                            Widget:=overlay:
                                                                Slots:=array:
                                                                    overlay_slot:
                                                                        HorizontalAlignment:=horizontal_alignment.Center
                                                                        VerticalAlignment:=vertical_alignment.Center
                                                                        Padding:=margin{Left:=7.0,Top:=PetsLineBoxPadding,Right:=PetsLineBoxPadding,Bottom:=PetsLineBoxPadding}
                                                                        Widget:=PetImage_AI_Slot5
                                                        overlay_slot:
                                                            HorizontalAlignment:=horizontal_alignment.Left
                                                            VerticalAlignment:=vertical_alignment.Center
                                                            Padding:=margin{Left:=TextPetsBoxPadding}
                                                            Widget:=stack_box:
                                                                Orientation:=orientation.Vertical
                                                                    Slots:=array:
                                                                        stack_box_slot: #Class
                                                                            HorizontalAlignment:=horizontal_alignment.Left
                                                                            VerticalAlignment:=vertical_alignment.Center
                                                                            Padding:=margin{Left:=TextPetsBoxPadding}
                                                                            Widget:=stack_box:
                                                                                Orientation:=orientation.Horizontal
                                                                                Slots:=array:
                                                                                    stack_box_slot: 
                                                                                        Padding:=margin{Left:=2.5,Right:=2.5}
                                                                                        Widget:=Class_AI_Slot5
                                                                                    stack_box_slot: 
                                                                                        Padding:=margin{Left:=2.5, Right:=1.0}
                                                                                        Widget:=PackA_AI_Box5
                                                                                    stack_box_slot: 
                                                                                        Padding:=margin{Right:=2.5}
                                                                                        Widget:=PackB_AI_Box5
                                                                        stack_box_slot: #Purity
                                                                            HorizontalAlignment:=horizontal_alignment.Left
                                                                            VerticalAlignment:=vertical_alignment.Center
                                                                            Padding:=margin{Left:=TextPetsBoxPadding}
                                                                            Widget:=Purity_AI_Slot5
                                                                        stack_box_slot: #Health
                                                                            HorizontalAlignment:=horizontal_alignment.Left
                                                                            VerticalAlignment:=vertical_alignment.Center
                                                                            Padding:=margin{Left:=TextPetsBoxPadding}
                                                                            Widget:=Health_AI_Slot5
                canvas_slot: #Challenges
                    Anchors:=anchors{Minimum:=vector2{X:=0.5,Y:=1.0},Maximum:=vector2{X:=0.5,Y:=1.0}}
                    ZOrder:=0
                    SizeToContent:=true
                    Offsets:=margin{Left:=0.0,Top:=0.0,Right:=0.0,Bottom:=0.0} #PositionX, PositionY, SizeX, SizeY
                    Alignment:=vector2{X:=0.5,Y:=1.0}
                    Widget:=stack_box:  
                        Orientation:=orientation.Vertical
                            Slots:=array:
                                stack_box_slot:
                                    HorizontalAlignment:=horizontal_alignment.Center
                                    VerticalAlignment:=vertical_alignment.Center
                                    Widget:=stack_box:  
                                        Orientation:=orientation.Horizontal
                                            Slots:=array:
                                                stack_box_slot:
                                                    Padding:=margin{Top:=0.0,Bottom:=0.0,Right:=2.5,Left:=2.5}
                                                    Widget:=text_block{DefaultText:=BasicMessage("Rewards:"), DefaultTextColor:=White, DefaultShadowOffset:=ShadowOffset, DefaultShadowOpacity := DefaultTextShadowOpacity}
                                                stack_box_slot:
                                                    HorizontalAlignment:=horizontal_alignment.Center
                                                    VerticalAlignment:=vertical_alignment.Center
                                                    Padding:=margin{Top:=0.0,Bottom:=0.0,Right:=2.5,Left:=2.5}
                                                    Widget:=overlay: 
                                                        Slots:=array:
                                                            overlay_slot:
                                                                Widget:=texture_block{DefaultDesiredSize:=vector2{X:=SizeBoxSaphiresX,Y:=SizeBoxSaphiresY},DefaultImage:=Images.EmptyTexture}
                                                            overlay_slot:
                                                                HorizontalAlignment:=horizontal_alignment.Fill
                                                                VerticalAlignment:=vertical_alignment.Fill
                                                                Widget:=texture_block{DefaultImage:=Images.Img_TintedBackgroundHighOpacity}  
                                                            overlay_slot:
                                                                HorizontalAlignment:=horizontal_alignment.Left
                                                                VerticalAlignment:=vertical_alignment.Center
                                                                Padding:=margin{Left:=Icon_LeftMargin}
                                                                Widget:=texture_block{DefaultDesiredSize:=vector2{X:=IconSmallSizeX,Y:=IconSmallSizeY},DefaultImage:=Images.Coal_Icon}
                                                            overlay_slot:
                                                                HorizontalAlignment:=horizontal_alignment.Left
                                                                VerticalAlignment:=vertical_alignment.Center
                                                                Padding:=margin{Left:=TextIcon_LeftMargin}
                                                                Widget:=CoalReward_Text
                                                stack_box_slot:
                                                    HorizontalAlignment:=horizontal_alignment.Center
                                                    VerticalAlignment:=vertical_alignment.Center
                                                    Padding:=margin{Top:=0.0,Bottom:=0.0,Right:=2.5,Left:=2.5}
                                                    Widget:=overlay: 
                                                        Slots:=array:
                                                            overlay_slot:
                                                                Widget:=texture_block{DefaultDesiredSize:=vector2{X:=SizeBoxSaphiresX,Y:=SizeBoxSaphiresY},DefaultImage:=Images.EmptyTexture}
                                                            overlay_slot:
                                                                HorizontalAlignment:=horizontal_alignment.Fill
                                                                VerticalAlignment:=vertical_alignment.Fill
                                                                Widget:=texture_block{DefaultImage:=Images.Img_TintedBackgroundHighOpacity} 
                                                            overlay_slot:
                                                                HorizontalAlignment:=horizontal_alignment.Left
                                                                VerticalAlignment:=vertical_alignment.Center
                                                                Padding:=margin{Left:=Icon_LeftMargin}
                                                                Widget:=texture_block{DefaultDesiredSize:=vector2{X:=IconSmallSizeX,Y:=IconSmallSizeY},DefaultImage:=Images.Mutant_Material_Icon}
                                                            overlay_slot:
                                                                HorizontalAlignment:=horizontal_alignment.Left
                                                                VerticalAlignment:=vertical_alignment.Center
                                                                Padding:=margin{Left:=TextIcon_LeftMargin}
                                                                Widget:=MutantMaterialReward_Text
                                stack_box_slot:
                                    HorizontalAlignment:=horizontal_alignment.Center
                                    VerticalAlignment:=vertical_alignment.Center
                                    Widget:=overlay: 
                                        Slots:=array:
                                            overlay_slot:
                                                Widget:=NumChallengesCompleted     
                                stack_box_slot:
                                    HorizontalAlignment:=horizontal_alignment.Center
                                    VerticalAlignment:=vertical_alignment.Center
                                    Widget:=stack_box:  
                                        Orientation:=orientation.Horizontal
                                            Slots:=array:
                                                stack_box_slot: #Slot 0
                                                    Padding:=margin{Top:=10.0,Bottom:=10.0,Right:=10.0,Left:=10.0} 
                                                    Widget:=stack_box: 
                                                        Orientation:=orientation.Vertical
                                                            Slots:=array:
                                                                stack_box_slot:
                                                                    Widget:=overlay: 
                                                                        Slots:=array:
                                                                            overlay_slot:
                                                                                Widget:=texture_block{DefaultImage:=Images.EmptyTexture, DefaultDesiredSize:=vector2{X:=350.0,Y:=1.0}}
                                                                            overlay_slot:
                                                                                HorizontalAlignment:=horizontal_alignment.Fill
                                                                                VerticalAlignment:=vertical_alignment.Fill
                                                                                Widget:=texture_block{DefaultImage:=Images.Img_TintedBackgroundHighOpacity}
                                                                            overlay_slot:
                                                                                HorizontalAlignment:=horizontal_alignment.Fill
                                                                                VerticalAlignment:=vertical_alignment.Fill
                                                                                Padding:=margin{Left:=SelectorBoxPadding,Right:=SelectorBoxPadding,Top:=SelectorBoxPadding,Bottom:=SelectorBoxPadding}
                                                                                Widget:=texture_block{DefaultImage:=Images.Img_TintedBackgroundHighOpacity}
                                                                            overlay_slot:
                                                                                Widget:=stack_box: 
                                                                                    Orientation:=orientation.Vertical
                                                                                        Slots:=array:
                                                                                            stack_box_slot:
                                                                                                Padding:=margin{Top:=5.0,Bottom:=5.0}
                                                                                                Widget:=SlotChallenge_Image1
                                                                                            stack_box_slot:
                                                                                                Padding:=margin{Top:=5.0,Bottom:=5.0}
                                                                                                Widget:=SlotChallenge_Description1
                                                                                            stack_box_slot:
                                                                                                Padding:=margin{Top:=5.0,Bottom:=5.0}
                                                                                                Widget:=SlotChallenge_IsCompleted1                                                      
                                                                stack_box_slot:
                                                                    Padding:=margin{Top:=5.0,Bottom:=5.0}
                                                                    Widget:=overlay:
                                                                        Slots:=array:
                                                                            overlay_slot:
                                                                                Widget:=texture_block{DefaultImage:=Images.EmptyTexture, DefaultDesiredSize:=vector2{X:=350.0,Y:=80.0}}
                                                                            overlay_slot:
                                                                                HorizontalAlignment:=horizontal_alignment.Fill
                                                                                VerticalAlignment:=vertical_alignment.Fill
                                                                                Widget:=ButtonChallenges5Left

                                                stack_box_slot: #Slot 1
                                                    Padding:=margin{Top:=10.0,Bottom:=10.0,Right:=10.0,Left:=10.0} 
                                                    Widget:=stack_box: 
                                                        Orientation:=orientation.Vertical
                                                            Slots:=array:
                                                                stack_box_slot:
                                                                    Widget:=overlay: 
                                                                        Slots:=array:
                                                                            overlay_slot:
                                                                                Widget:=texture_block{DefaultImage:=Images.EmptyTexture, DefaultDesiredSize:=vector2{X:=350.0,Y:=1.0}}
                                                                            overlay_slot:
                                                                                HorizontalAlignment:=horizontal_alignment.Fill
                                                                                VerticalAlignment:=vertical_alignment.Fill
                                                                                Widget:=texture_block{DefaultImage:=Images.Img_TintedBackgroundHighOpacity}
                                                                            overlay_slot:
                                                                                HorizontalAlignment:=horizontal_alignment.Fill
                                                                                VerticalAlignment:=vertical_alignment.Fill
                                                                                Padding:=margin{Left:=SelectorBoxPadding,Right:=SelectorBoxPadding,Top:=SelectorBoxPadding,Bottom:=SelectorBoxPadding}
                                                                                Widget:=texture_block{DefaultImage:=Images.Img_TintedBackgroundHighOpacity}
                                                                            overlay_slot:
                                                                                Widget:=stack_box: 
                                                                                    Orientation:=orientation.Vertical
                                                                                        Slots:=array:
                                                                                            stack_box_slot:
                                                                                                Padding:=margin{Top:=5.0,Bottom:=5.0}
                                                                                                Widget:=SlotChallenge_Image2
                                                                                            stack_box_slot:
                                                                                                Padding:=margin{Top:=5.0,Bottom:=5.0}
                                                                                                Widget:=SlotChallenge_Description2
                                                                                            stack_box_slot:
                                                                                                Padding:=margin{Top:=5.0,Bottom:=5.0}
                                                                                                Widget:=SlotChallenge_IsCompleted2                                                      
                                                                stack_box_slot:
                                                                    Padding:=margin{Top:=5.0,Bottom:=5.0}
                                                                    Widget:=overlay:
                                                                        Slots:=array:
                                                                            overlay_slot:
                                                                                Widget:=texture_block{DefaultImage:=Images.EmptyTexture, DefaultDesiredSize:=vector2{X:=350.0,Y:=80.0}}
                                                                            overlay_slot:
                                                                                HorizontalAlignment:=horizontal_alignment.Fill
                                                                                VerticalAlignment:=vertical_alignment.Fill
                                                                                Widget:=ButtonChallengesLeft

                                                stack_box_slot: #Slot 2
                                                    Padding:=margin{Top:=10.0,Bottom:=10.0,Right:=10.0,Left:=10.0} 
                                                    Widget:=stack_box: 
                                                        Orientation:=orientation.Vertical
                                                            Slots:=array:
                                                                stack_box_slot:
                                                                    Widget:=overlay: 
                                                                        Slots:=array:
                                                                            overlay_slot:
                                                                                Widget:=texture_block{DefaultImage:=Images.EmptyTexture, DefaultDesiredSize:=vector2{X:=350.0,Y:=1.0}}
                                                                            overlay_slot:
                                                                                HorizontalAlignment:=horizontal_alignment.Fill
                                                                                VerticalAlignment:=vertical_alignment.Fill
                                                                                Widget:=texture_block{DefaultImage:=Images.Img_TintedBackgroundHighOpacity}
                                                                            overlay_slot:
                                                                                HorizontalAlignment:=horizontal_alignment.Fill
                                                                                VerticalAlignment:=vertical_alignment.Fill
                                                                                Padding:=margin{Left:=SelectorBoxPadding,Right:=SelectorBoxPadding,Top:=SelectorBoxPadding,Bottom:=SelectorBoxPadding}
                                                                                Widget:=texture_block{DefaultImage:=Images.Img_TintedBackgroundHighOpacity}
                                                                            overlay_slot:
                                                                                Widget:=stack_box: 
                                                                                    Orientation:=orientation.Vertical
                                                                                        Slots:=array:
                                                                                            stack_box_slot:
                                                                                                Padding:=margin{Top:=5.0,Bottom:=5.0}
                                                                                                Widget:=SlotChallenge_Image3
                                                                                            stack_box_slot:
                                                                                                Padding:=margin{Top:=5.0,Bottom:=5.0}
                                                                                                Widget:=SlotChallenge_Description3
                                                                                            stack_box_slot:
                                                                                                Padding:=margin{Top:=5.0,Bottom:=5.0}
                                                                                                Widget:=SlotChallenge_IsCompleted3                                                     
                                                                stack_box_slot:
                                                                    Padding:=margin{Top:=5.0,Bottom:=5.0}
                                                                    Widget:=overlay:
                                                                        Slots:=array:
                                                                            overlay_slot:
                                                                                Widget:=texture_block{DefaultImage:=Images.EmptyTexture, DefaultDesiredSize:=vector2{X:=350.0,Y:=80.0}}
                                                                            overlay_slot:
                                                                                HorizontalAlignment:=horizontal_alignment.Fill
                                                                                VerticalAlignment:=vertical_alignment.Fill
                                                                                Widget:=ButtonChallengesOK
                                                stack_box_slot: #Slot 4
                                                    Padding:=margin{Top:=10.0,Bottom:=10.0,Right:=10.0,Left:=10.0} 
                                                    Widget:=stack_box: 
                                                        Orientation:=orientation.Vertical
                                                            Slots:=array:
                                                                stack_box_slot:
                                                                    Widget:=overlay: 
                                                                        Slots:=array:
                                                                            overlay_slot:
                                                                                Widget:=texture_block{DefaultImage:=Images.EmptyTexture, DefaultDesiredSize:=vector2{X:=350.0,Y:=1.0}}
                                                                            overlay_slot:
                                                                                HorizontalAlignment:=horizontal_alignment.Fill
                                                                                VerticalAlignment:=vertical_alignment.Fill
                                                                                Widget:=texture_block{DefaultImage:=Images.Img_TintedBackgroundHighOpacity}
                                                                            overlay_slot:
                                                                                HorizontalAlignment:=horizontal_alignment.Fill
                                                                                VerticalAlignment:=vertical_alignment.Fill
                                                                                Padding:=margin{Left:=SelectorBoxPadding,Right:=SelectorBoxPadding,Top:=SelectorBoxPadding,Bottom:=SelectorBoxPadding}
                                                                                Widget:=texture_block{DefaultImage:=Images.Img_TintedBackgroundHighOpacity}
                                                                            overlay_slot:
                                                                                Widget:=stack_box: 
                                                                                    Orientation:=orientation.Vertical
                                                                                        Slots:=array:
                                                                                            stack_box_slot:
                                                                                                Padding:=margin{Top:=5.0,Bottom:=5.0}
                                                                                                Widget:=SlotChallenge_Image4
                                                                                            stack_box_slot:
                                                                                                Padding:=margin{Top:=5.0,Bottom:=5.0}
                                                                                                Widget:=SlotChallenge_Description4
                                                                                            stack_box_slot:
                                                                                                Padding:=margin{Top:=5.0,Bottom:=5.0}
                                                                                                Widget:=SlotChallenge_IsCompleted4                                                      
                                                                stack_box_slot:
                                                                    Padding:=margin{Top:=5.0,Bottom:=5.0}
                                                                    Widget:=overlay:
                                                                        Slots:=array:
                                                                            overlay_slot:
                                                                                Widget:=texture_block{DefaultImage:=Images.EmptyTexture, DefaultDesiredSize:=vector2{X:=350.0,Y:=80.0}}
                                                                            overlay_slot:
                                                                                HorizontalAlignment:=horizontal_alignment.Fill
                                                                                VerticalAlignment:=vertical_alignment.Fill
                                                                                Widget:=ButtonChallengesRight

                                                stack_box_slot: #Slot 5
                                                    Padding:=margin{Top:=10.0,Bottom:=10.0,Right:=10.0,Left:=10.0} 
                                                    Widget:=stack_box: 
                                                        Orientation:=orientation.Vertical
                                                            Slots:=array:
                                                                stack_box_slot:
                                                                    Widget:=overlay: 
                                                                        Slots:=array:
                                                                            overlay_slot:
                                                                                Widget:=texture_block{DefaultImage:=Images.EmptyTexture, DefaultDesiredSize:=vector2{X:=350.0,Y:=1.0}}
                                                                            overlay_slot:
                                                                                HorizontalAlignment:=horizontal_alignment.Fill
                                                                                VerticalAlignment:=vertical_alignment.Fill
                                                                                Widget:=texture_block{DefaultImage:=Images.Img_TintedBackgroundHighOpacity}
                                                                            overlay_slot:
                                                                                HorizontalAlignment:=horizontal_alignment.Fill
                                                                                VerticalAlignment:=vertical_alignment.Fill
                                                                                Padding:=margin{Left:=SelectorBoxPadding,Right:=SelectorBoxPadding,Top:=SelectorBoxPadding,Bottom:=SelectorBoxPadding}
                                                                                Widget:=texture_block{DefaultImage:=Images.Img_TintedBackgroundHighOpacity}
                                                                            overlay_slot:
                                                                                Widget:=stack_box: 
                                                                                    Orientation:=orientation.Vertical
                                                                                        Slots:=array:
                                                                                            stack_box_slot:
                                                                                                Padding:=margin{Top:=5.0,Bottom:=5.0}
                                                                                                Widget:=SlotChallenge_Image5
                                                                                            stack_box_slot:
                                                                                                Padding:=margin{Top:=5.0,Bottom:=5.0}
                                                                                                Widget:=SlotChallenge_Description5
                                                                                            stack_box_slot:
                                                                                                Padding:=margin{Top:=5.0,Bottom:=5.0}
                                                                                                Widget:=SlotChallenge_IsCompleted5                                                      
                                                                stack_box_slot:
                                                                    Padding:=margin{Top:=5.0,Bottom:=5.0}
                                                                    Widget:=overlay:
                                                                        Slots:=array:
                                                                            overlay_slot:
                                                                                Widget:=texture_block{DefaultImage:=Images.EmptyTexture, DefaultDesiredSize:=vector2{X:=350.0,Y:=80.0}}
                                                                            overlay_slot:
                                                                                HorizontalAlignment:=horizontal_alignment.Fill
                                                                                VerticalAlignment:=vertical_alignment.Fill
                                                                                Widget:=ButtonChallenges5Right

                                                                                


                                                    

                                                <#
                                                stack_box_slot: #Slot 0
                                                    Padding:=margin{Top:=10.0,Bottom:=10.0,Right:=10.0,Left:=10.0} 
                                                    Widget:=overlay: 
                                                        Slots:=array:
                                                            overlay_slot:
                                                                HorizontalAlignment:=horizontal_alignment.Fill
                                                                VerticalAlignment:=vertical_alignment.Fill
                                                                Widget:=texture_block{DefaultImage:=Images.Img_TintedBackgroundHighOpacity}
                                                            overlay_slot:
                                                                HorizontalAlignment:=horizontal_alignment.Fill
                                                                VerticalAlignment:=vertical_alignment.Fill
                                                                Padding:=margin{Left:=SelectorBoxPadding,Right:=SelectorBoxPadding,Top:=SelectorBoxPadding,Bottom:=SelectorBoxPadding}
                                                                Widget:=texture_block{DefaultImage:=Images.Img_TintedBackgroundHighOpacity}
                                                            overlay_slot:
                                                                Widget:=stack_box: 
                                                                    Orientation:=orientation.Vertical
                                                                        Slots:=array:
                                                                            stack_box_slot:
                                                                                Padding:=margin{Top:=5.0,Bottom:=5.0}
                                                                                Widget:=SlotChallenge_Image1
                                                                            stack_box_slot:
                                                                                Padding:=margin{Top:=5.0,Bottom:=5.0}
                                                                                Widget:=SlotChallenge_Description1
                                                                            stack_box_slot:
                                                                                Padding:=margin{Top:=5.0,Bottom:=5.0}
                                                                                Widget:=SlotChallenge_IsCompleted1
                                                                            stack_box_slot:
                                                                                Padding:=margin{Top:=5.0,Bottom:=5.0}
                                                                                Widget:=overlay:
                                                                                    Slots:=array:
                                                                                        overlay_slot:
                                                                                            Widget:=texture_block{DefaultImage:=Images.EmptyTexture, DefaultDesiredSize:=vector2{X:=350.0,Y:=100.0}}
                                                stack_box_slot: #Slot 1
                                                    Padding:=margin{Top:=10.0,Bottom:=10.0,Right:=10.0,Left:=10.0} 
                                                    Widget:=overlay: 
                                                        Slots:=array:
                                                            overlay_slot:
                                                                HorizontalAlignment:=horizontal_alignment.Fill
                                                                VerticalAlignment:=vertical_alignment.Fill
                                                                Widget:=texture_block{DefaultImage:=Images.Img_TintedBackgroundHighOpacity}
                                                            overlay_slot:
                                                                HorizontalAlignment:=horizontal_alignment.Fill
                                                                VerticalAlignment:=vertical_alignment.Fill
                                                                Padding:=margin{Left:=SelectorBoxPadding,Right:=SelectorBoxPadding,Top:=SelectorBoxPadding,Bottom:=SelectorBoxPadding}
                                                                Widget:=texture_block{DefaultImage:=Images.Img_TintedBackgroundHighOpacity}
                                                            overlay_slot:
                                                                Widget:=stack_box: 
                                                                    Orientation:=orientation.Vertical
                                                                        Slots:=array:
                                                                            stack_box_slot:
                                                                                Padding:=margin{Top:=5.0,Bottom:=5.0}
                                                                                Widget:=SlotChallenge_Image2
                                                                            stack_box_slot:
                                                                                Padding:=margin{Top:=5.0,Bottom:=5.0}
                                                                                Widget:=SlotChallenge_Description2
                                                                            stack_box_slot:
                                                                                Padding:=margin{Top:=5.0,Bottom:=5.0}
                                                                                Widget:=SlotChallenge_IsCompleted2
                                                                            stack_box_slot:
                                                                                Padding:=margin{Top:=5.0,Bottom:=5.0}
                                                                                Widget:=overlay:
                                                                                    Slots:=array:
                                                                                        overlay_slot:
                                                                                            Widget:=texture_block{DefaultImage:=Images.EmptyTexture, DefaultDesiredSize:=vector2{X:=350.0,Y:=100.0}}
                                                                                        overlay_slot:
                                                                                            HorizontalAlignment:=horizontal_alignment.Fill
                                                                                            Widget:=ButtonChallengesLeft
                                                stack_box_slot: #Slot 2
                                                    Padding:=margin{Top:=10.0,Bottom:=10.0,Right:=10.0,Left:=10.0} 
                                                    Widget:=overlay: 
                                                        Slots:=array:
                                                            overlay_slot:
                                                                HorizontalAlignment:=horizontal_alignment.Fill
                                                                VerticalAlignment:=vertical_alignment.Fill
                                                                Widget:=texture_block{DefaultImage:=Images.Img_TintedBackgroundHighOpacity}
                                                            overlay_slot:
                                                                HorizontalAlignment:=horizontal_alignment.Fill
                                                                VerticalAlignment:=vertical_alignment.Fill
                                                                Padding:=margin{Left:=SelectorBoxPadding,Right:=SelectorBoxPadding,Top:=SelectorBoxPadding,Bottom:=SelectorBoxPadding}
                                                                Widget:=texture_block{DefaultImage:=Images.Img_TintedBackgroundHighOpacity}
                                                            overlay_slot:
                                                                Widget:=stack_box: 
                                                                    Orientation:=orientation.Vertical
                                                                        Slots:=array:
                                                                            stack_box_slot:
                                                                                Padding:=margin{Top:=5.0,Bottom:=5.0}
                                                                                Widget:=SlotChallenge_Image3
                                                                            stack_box_slot:
                                                                                Padding:=margin{Top:=5.0,Bottom:=5.0}
                                                                                Widget:=SlotChallenge_Description3
                                                                            stack_box_slot:
                                                                                Padding:=margin{Top:=5.0,Bottom:=5.0}
                                                                                Widget:=SlotChallenge_IsCompleted3
                                                                            stack_box_slot:
                                                                                Padding:=margin{Top:=5.0,Bottom:=5.0}
                                                                                Widget:=overlay:
                                                                                    Slots:=array:
                                                                                        overlay_slot:
                                                                                            Widget:=texture_block{DefaultImage:=Images.EmptyTexture, DefaultDesiredSize:=vector2{X:=350.0,Y:=100.0}}
                                                                                        overlay_slot:
                                                                                            HorizontalAlignment:=horizontal_alignment.Fill
                                                                                            Widget:=ButtonChallengesOK
                                                stack_box_slot: #Slot 3
                                                    Padding:=margin{Top:=10.0,Bottom:=10.0,Right:=10.0,Left:=10.0} 
                                                    Widget:=overlay: 
                                                        Slots:=array:
                                                            overlay_slot:
                                                                HorizontalAlignment:=horizontal_alignment.Fill
                                                                VerticalAlignment:=vertical_alignment.Fill
                                                                Widget:=texture_block{DefaultImage:=Images.Img_TintedBackgroundHighOpacity}
                                                            overlay_slot:
                                                                HorizontalAlignment:=horizontal_alignment.Fill
                                                                VerticalAlignment:=vertical_alignment.Fill
                                                                Padding:=margin{Left:=SelectorBoxPadding,Right:=SelectorBoxPadding,Top:=SelectorBoxPadding,Bottom:=SelectorBoxPadding}
                                                                Widget:=texture_block{DefaultImage:=Images.Img_TintedBackgroundHighOpacity}
                                                            overlay_slot:
                                                                Widget:=stack_box: 
                                                                    Orientation:=orientation.Vertical
                                                                        Slots:=array:
                                                                            stack_box_slot:
                                                                                Padding:=margin{Top:=5.0,Bottom:=5.0}
                                                                                Widget:=SlotChallenge_Image4
                                                                            stack_box_slot:
                                                                                Padding:=margin{Top:=5.0,Bottom:=5.0}
                                                                                Widget:=SlotChallenge_Description4
                                                                            stack_box_slot:
                                                                                Padding:=margin{Top:=5.0,Bottom:=5.0}
                                                                                Widget:=SlotChallenge_IsCompleted4
                                                                            stack_box_slot:
                                                                                Padding:=margin{Top:=5.0,Bottom:=5.0}
                                                                                Widget:=overlay:
                                                                                    Slots:=array:
                                                                                        overlay_slot:
                                                                                            Widget:=texture_block{DefaultImage:=Images.EmptyTexture, DefaultDesiredSize:=vector2{X:=350.0,Y:=100.0}}
                                                                                        overlay_slot:
                                                                                            HorizontalAlignment:=horizontal_alignment.Fill
                                                                                            Widget:=ButtonChallengesRight
                                                stack_box_slot: #Slot 4
                                                    Padding:=margin{Top:=10.0,Bottom:=10.0,Right:=10.0,Left:=10.0}
                                                    Widget:=overlay: 
                                                        Slots:=array:
                                                            overlay_slot:
                                                                HorizontalAlignment:=horizontal_alignment.Fill
                                                                VerticalAlignment:=vertical_alignment.Fill
                                                                Widget:=texture_block{DefaultImage:=Images.Img_TintedBackgroundHighOpacity}
                                                            overlay_slot:
                                                                HorizontalAlignment:=horizontal_alignment.Fill
                                                                VerticalAlignment:=vertical_alignment.Fill
                                                                Padding:=margin{Left:=SelectorBoxPadding,Right:=SelectorBoxPadding,Top:=SelectorBoxPadding,Bottom:=SelectorBoxPadding}
                                                                Widget:=texture_block{DefaultImage:=Images.Img_TintedBackgroundHighOpacity}
                                                            overlay_slot:
                                                                Widget:=stack_box: 
                                                                    Orientation:=orientation.Vertical
                                                                        Slots:=array:
                                                                            stack_box_slot:
                                                                                Padding:=margin{Top:=5.0,Bottom:=5.0}
                                                                                Widget:=SlotChallenge_Image5
                                                                            stack_box_slot:
                                                                                Padding:=margin{Top:=5.0,Bottom:=5.0}
                                                                                Widget:=SlotChallenge_Description5
                                                                            stack_box_slot:
                                                                                Padding:=margin{Top:=5.0,Bottom:=5.0}
                                                                                Widget:=SlotChallenge_IsCompleted5
                                                                            stack_box_slot:
                                                                                Padding:=margin{Top:=5.0,Bottom:=5.0}
                                                                                Widget:=overlay:
                                                                                    Slots:=array:
                                                                                        overlay_slot:
                                                                                            Widget:=texture_block{DefaultImage:=Images.EmptyTexture, DefaultDesiredSize:=vector2{X:=350.0,Y:=100.0}}   #>                                    
                                stack_box_slot:
                                    Widget:=overlay: 
                                        Slots:=array:
                                            overlay_slot:
                                                Widget:=texture_block{DefaultImage:=Images.EmptyTexture, DefaultDesiredSize:=vector2{X:=350.0,Y:=80.0}}
                                            overlay_slot:
                                                Widget:= ButtonChallengesClose
                                                HorizontalAlignment:=horizontal_alignment.Fill
                                                VerticalAlignment:=vertical_alignment.Fill

            


                                    
                
        

