# fill this file with your verse-script
using { /UnrealEngine.com/Temporary/UI }
using { /UnrealEngine.com/Temporary/SpatialMath }
using { /Fortnite.com/UI }
using { /Verse.org/Simulation }
using { /Verse.org/Colors/NamedColors }

# A Verse-authored creative device that can be placed in a level
hexa_minigame_selector_ui := class():

    #Class settings
    var MaybeAgent : ?agent = false
    HexaMinigameManager : hexa_minigame_manager

    var MaybeCanvas : ?canvas = false
    var UIIsDisplayed : logic = false

    #Buttons
    ResetButton : button_loud = button_loud{}
    SelectButton : button_regular = button_regular{}
    NavigateRightButton : button_quiet = button_quiet{}
    NavigateLeftButton : button_quiet = button_quiet{}
    CancelButton : button_loud = button_loud{}
    FightButton : button_regular = button_regular{}

    var CurrentOption : int = 0
    var ConfirmationToReset : int = 0

    QuietButton_PetSelectorDesiredSizeV2 : vector2 = vector2{X:=150.0,Y:=75.0}

    #Box0
    TextureBlock_PetImages_Box0 : texture_block = texture_block{DefaultImage:=Images.EmptyTexture,DefaultDesiredSize:=PetsSelectorImageSize}
    TextBlock_Classes_Box0 : text_block = text_block{DefaultTextColor := White,DefaultShadowOffset:=ShadowOffset, DefaultShadowOpacity := 1.0}
    TextBlock_PackA_Box0 : text_block = text_block{DefaultTextColor := White,DefaultShadowOffset:=ShadowOffset, DefaultShadowOpacity := 1.0}
    TextBlock_PackB_Box0 : text_block = text_block{DefaultTextColor := White,DefaultShadowOffset:=ShadowOffset, DefaultShadowOpacity := 1.0}
    TextBlock_Purity_Box0 : text_block = text_block{DefaultTextColor := White,DefaultShadowOffset:=ShadowOffset, DefaultShadowOpacity := 1.0}
    TextBlock_Tier_Box0 : text_block = text_block{DefaultTextColor := White,DefaultShadowOffset:=ShadowOffset, DefaultShadowOpacity := 1.0}

    #Box1
    TextureBlock_PetImages_Box1 : texture_block = texture_block{DefaultImage:=Images.EmptyTexture,DefaultDesiredSize:=PetsSelectorImageSize}
    TextBlock_Classes_Box1 : text_block = text_block{DefaultTextColor := White,DefaultShadowOffset:=ShadowOffset, DefaultShadowOpacity := 1.0}
    TextBlock_PackA_Box1 : text_block = text_block{DefaultTextColor := White,DefaultShadowOffset:=ShadowOffset, DefaultShadowOpacity := 1.0}
    TextBlock_PackB_Box1 : text_block = text_block{DefaultTextColor := White,DefaultShadowOffset:=ShadowOffset, DefaultShadowOpacity := 1.0}
    TextBlock_Purity_Box1 : text_block = text_block{DefaultTextColor := White,DefaultShadowOffset:=ShadowOffset, DefaultShadowOpacity := 1.0}
    TextBlock_Tier_Box1 : text_block = text_block{DefaultTextColor := White,DefaultShadowOffset:=ShadowOffset, DefaultShadowOpacity := 1.0}

    #Box2
    TextureBlock_PetImages_Box2 : texture_block = texture_block{DefaultImage:=Images.EmptyTexture,DefaultDesiredSize:=PetsSelectorImageSize}
    TextBlock_Classes_Box2 : text_block = text_block{DefaultTextColor := White,DefaultShadowOffset:=ShadowOffset, DefaultShadowOpacity := 1.0}
    TextBlock_PackA_Box2 : text_block = text_block{DefaultTextColor := White,DefaultShadowOffset:=ShadowOffset, DefaultShadowOpacity := 1.0}
    TextBlock_PackB_Box2 : text_block = text_block{DefaultTextColor := White,DefaultShadowOffset:=ShadowOffset, DefaultShadowOpacity := 1.0}
    TextBlock_Purity_Box2 : text_block = text_block{DefaultTextColor := White,DefaultShadowOffset:=ShadowOffset, DefaultShadowOpacity := 1.0}
    TextBlock_Tier_Box2 : text_block = text_block{DefaultTextColor := White,DefaultShadowOffset:=ShadowOffset, DefaultShadowOpacity := 1.0}

    #Box3
    TextureBlock_PetImages_Box3 : texture_block = texture_block{DefaultImage:=Images.EmptyTexture,DefaultDesiredSize:=PetsSelectorImageSize}
    TextBlock_Classes_Box3 : text_block = text_block{DefaultTextColor := White,DefaultShadowOffset:=ShadowOffset, DefaultShadowOpacity := 1.0}
    TextBlock_PackA_Box3 : text_block = text_block{DefaultTextColor := White,DefaultShadowOffset:=ShadowOffset, DefaultShadowOpacity := 1.0}
    TextBlock_PackB_Box3 : text_block = text_block{DefaultTextColor := White,DefaultShadowOffset:=ShadowOffset, DefaultShadowOpacity := 1.0}
    TextBlock_Purity_Box3 : text_block = text_block{DefaultTextColor := White,DefaultShadowOffset:=ShadowOffset, DefaultShadowOpacity := 1.0}
    TextBlock_Tier_Box3 : text_block = text_block{DefaultTextColor := White,DefaultShadowOffset:=ShadowOffset, DefaultShadowOpacity := 1.0}

    #All Arrays
    var AllTextureBlock_PetImages : []texture_block = array{}
    var AllTextBlock_Classes : []text_block = array{}
    var AllTextBlock_LVPackageA : []text_block = array{}
    var AllTextBlock_LVPackageB : []text_block = array{}
    var AllTextBlock_Purity : []text_block = array{}
    var AllTextBlock_Tier : []text_block = array{}
    
    var Selected_TextureBlock_PetImages : texture_block = texture_block{DefaultImage:=Images.EmptyTexture,DefaultDesiredSize:=PetsSelectorImageSize}
    var Selected_TextBlock_Class : text_block = text_block{DefaultTextColor := White,DefaultShadowOffset:=ShadowOffset, DefaultShadowOpacity := 1.0}
    var Selected_TextBlock_LVPackageA : text_block = text_block{DefaultTextColor := White,DefaultShadowOffset:=ShadowOffset, DefaultShadowOpacity := 1.0}
    var Selected_TextBlock_LVPackageB : text_block = text_block{DefaultTextColor := White,DefaultShadowOffset:=ShadowOffset, DefaultShadowOpacity := 1.0}
    var Selected_TextBlock_Purity : text_block = text_block{DefaultTextColor := White,DefaultShadowOffset:=ShadowOffset, DefaultShadowOpacity := 1.0}
    var Selected_TextBlock_Health : text_block = text_block{DefaultTextColor := White,DefaultShadowOffset:=ShadowOffset, DefaultShadowOpacity := 1.0}
    var Selected_TextBlock_Damage : text_block = text_block{DefaultTextColor := White,DefaultShadowOffset:=ShadowOffset, DefaultShadowOpacity := 1.0}
    var Selected_TextBlock_Healing : text_block = text_block{DefaultTextColor := White,DefaultShadowOffset:=ShadowOffset, DefaultShadowOpacity := 1.0}
    var Selected_TextBlock_AttackSpeed : text_block = text_block{DefaultTextColor := White,DefaultShadowOffset:=ShadowOffset, DefaultShadowOpacity := 1.0}
    var Selected_TextBlock_Speed : text_block = text_block{DefaultTextColor := White,DefaultShadowOffset:=ShadowOffset, DefaultShadowOpacity := 1.0}
    var Selected_TextBlock_Range : text_block = text_block{DefaultTextColor := White,DefaultShadowOffset:=ShadowOffset, DefaultShadowOpacity := 1.0}


    #Close
    CloseButton : button_loud = button_loud{}

    InitUI() : void =
        Canvas := MakeCanvas()
        set MaybeCanvas = option{Canvas}

        FightButton.SetText(BasicMessage("Fight"))
        SelectButton.SetText(BasicMessage("Select"))
        NavigateRightButton.SetText(BasicMessage("▶"))
        NavigateLeftButton.SetText(BasicMessage("◀"))
        CancelButton.SetText(BasicMessage("Cancel"))
        ResetButton.SetText(BasicMessage("RESET ALL"))

        set AllTextureBlock_PetImages = array{TextureBlock_PetImages_Box0,TextureBlock_PetImages_Box1,texture_block{DefaultImage := Images.EmptyTexture},TextureBlock_PetImages_Box2,TextureBlock_PetImages_Box3}
        set AllTextBlock_Classes = array{TextBlock_Classes_Box0,TextBlock_Classes_Box1,text_block{},TextBlock_Classes_Box2,TextBlock_Classes_Box3}
        set AllTextBlock_LVPackageA = array{TextBlock_PackA_Box0,TextBlock_PackA_Box1,text_block{},TextBlock_PackA_Box2,TextBlock_PackA_Box3}
        set AllTextBlock_LVPackageB = array{TextBlock_PackB_Box0,TextBlock_PackB_Box1,text_block{},TextBlock_PackB_Box2,TextBlock_PackB_Box3}
        set AllTextBlock_Purity = array{TextBlock_Purity_Box0,TextBlock_Purity_Box1,text_block{},TextBlock_Purity_Box2,TextBlock_Purity_Box3}
        set AllTextBlock_Tier = array{TextBlock_Tier_Box0,TextBlock_Tier_Box1,text_block{},TextBlock_Tier_Box2,TextBlock_Tier_Box3}

        Subscribes()

    SetAgent(InAgent : agent) : void =
        set MaybeAgent = option{InAgent}
        set CurrentOption = 0

    HandlerSelectPet(Message : widget_message) : void =
        if(HexaPet:=HexaMinigameManager.AllHexaPlayerPets[CurrentOption]):
            if(HexaPet.IsBeingUsed?):
                HexaMinigameManager.WithdrawPet(HexaPet,HexaPet.Position)
                RefreshAll()
            else:
                Remove_CanvasWithoutMessage()
                HexaMinigameManager.SelectPet(CurrentOption)

    NavidateRight(Message : widget_message) : void =
        set CurrentOption = CurrentOption + 1
        if(HexaMinigameManager.AllHexaPlayerPets.Length <= CurrentOption):
            set CurrentOption = HexaMinigameManager.AllHexaPlayerPets.Length-1
            #Print("ERROR, no more pets on this side")
        RefreshAll()

    NavidateLeft(Message : widget_message) : void =
        set CurrentOption = CurrentOption - 1
        if(CurrentOption < 0):
            set CurrentOption = 0
            #Print("ERROR, no more pets on this side")
        RefreshAll()

           
    Subscribes() : void =
        ResetButton.OnClick().Subscribe(ResetPlayerPets)
        FightButton.OnClick().Subscribe(StartFight)
        SelectButton.OnClick().Subscribe(HandlerSelectPet)
        NavigateRightButton.OnClick().Subscribe(NavidateRight)
        NavigateLeftButton.OnClick().Subscribe(NavidateLeft)
        CancelButton.OnClick().Subscribe(Handler_PetSelector_ToChallengeSelector)

    ResetPlayerPets(Message : widget_message) : void =
        if(ConfirmationToReset=0):
            set ConfirmationToReset = ConfirmationToReset + 1
            ResetButton.SetText(BasicMessage("ARE YOU SURE?"))
        else:
            set ConfirmationToReset = 0
            ResetButton.SetText(BasicMessage("RESET ALL"))
            HexaMinigameManager.ResetAgentPets()

    StartFight(Message : widget_message) : void =
        var IsTherePetUsed : logic = false
        for(HexaPet : HexaMinigameManager.AllHexaPlayerPets, HexaPet.IsBeingUsed?):
            set IsTherePetUsed = true
        if(IsTherePetUsed?):
            spawn:
                DelayChangeUI_Selector_Fight(Message)
        else:
            spawn:
                SpawnWarningNoPets()

    SpawnWarningNoPets()<suspends> : void =
        FightButton.SetText(BasicMessage("PLACE SOURDOUGH FIRST"))
        Sleep(1.0)
        FightButton.SetText(BasicMessage("FIGHT"))
        
    DelayChangeUI_Selector_Fight(Message : widget_message)<suspends> : void =
        if(BakeryManager := HexaMinigameManager.MaybeBakeryMachinesManager?, Message.Player.IsActive[], BakeryManager.PlayerDataCache.HexaGame_Accumulated>0):
            FightButton.SetText(BasicMessage("FIGHT"))
            Remove_Canvas(Message)
            Sleep(0.1)
            HexaMinigameManager.StartFight(Message)
        else:
            FightButton.SetText(BasicMessage("NO COMPETITION TOKENS"))

    RefreshAll() : void =
        FightButton.SetText(BasicMessage("FIGHT"))
        AllHexaPets:=HexaMinigameManager.AllHexaPlayerPets
        if(Pet := AllHexaPets[CurrentOption]):
            SetPetImage(Pet.PetType, Selected_TextureBlock_PetImages)
            SetPetPurityAndTier(Pet.Purity,Pet.Tier,Selected_TextBlock_Purity)
            SetPetClass(Pet,Selected_TextBlock_Class,Selected_TextBlock_LVPackageA,Selected_TextBlock_LVPackageB)
        
            Selected_TextBlock_Health.SetText(BasicMessage("❤️ {GetDoubleDecimalMagnitudeLetter(Pet.Health,"","")}"))
            Selected_TextBlock_Damage.SetText(BasicMessage("💥 {GetDoubleDecimalMagnitudeLetter(Pet.AttackDamage,"","")}"))
            if(Pet.PetType = DRUID):
                Selected_TextBlock_Healing.SetText(BasicMessage("💉 {GetDoubleDecimalMagnitudeLetter(Pet.AttackDamage,"","")} / s"))
            else:
                Selected_TextBlock_Healing.SetText(BasicMessage("-"))
            Selected_TextBlock_AttackSpeed.SetText(BasicMessage("🔥 {GetDoubleDecimalMagnitudeLetter(Pet.ActionSpeed,"","")} / s"))
            Selected_TextBlock_Speed.SetText(BasicMessage("🏃 {GetDoubleDecimalMagnitudeLetter(Pet.Speed,"","")} Tiles / s"))

            TilesRange:=ConvertRange(false, Pet.AttackRange*1.0)

            Selected_TextBlock_Range.SetText(BasicMessage("🏹 {GetDoubleDecimalMagnitudeLetter(TilesRange,"","")} Tiles"))

            if(Pet.IsBeingUsed = true):
                Selected_TextureBlock_PetImages.SetTint(Red)
                Selected_TextBlock_Class.SetTextColor(Black)
                Selected_TextBlock_LVPackageA.SetTextColor(Black)
                Selected_TextBlock_LVPackageB.SetTextColor(Black)
                Selected_TextBlock_Purity.SetTextColor(Black)
                Selected_TextBlock_Health.SetTextColor(Black)
                Selected_TextBlock_Damage.SetTextColor(Black)
                Selected_TextBlock_Healing.SetTextColor(Black)
                Selected_TextBlock_AttackSpeed.SetTextColor(Black)
                Selected_TextBlock_Speed.SetTextColor(Black)
                Selected_TextBlock_Range.SetTextColor(Black)
            else:
                Selected_TextureBlock_PetImages.SetTint(White)
                Selected_TextBlock_Class.SetTextColor(White)
                Selected_TextBlock_Purity.SetTextColor(White)
                Selected_TextBlock_Health.SetTextColor(White)
                Selected_TextBlock_Damage.SetTextColor(White)
                Selected_TextBlock_Healing.SetTextColor(White)
                Selected_TextBlock_AttackSpeed.SetTextColor(White)
                Selected_TextBlock_Speed.SetTextColor(White)
                Selected_TextBlock_Range.SetTextColor(White)

        for(Index:=0..AllTextureBlock_PetImages.Length-1):
            if:
                TextureBlock_PetImage:=AllTextureBlock_PetImages[Index]
                TextBlock_Class:=AllTextBlock_Classes[Index]
                TextBlock_LVPackageA:=AllTextBlock_LVPackageA[Index]
                TextBlock_LVPackageB:=AllTextBlock_LVPackageB[Index]
                TextBlock_Purity:=AllTextBlock_Purity[Index]
                TextBlock_Tier:=AllTextBlock_Tier[Index]
            then:
                if(PetIndexSelected:=AllHexaPets[CurrentOption-2+Index]):
                    SetPetImage(PetIndexSelected.PetType, TextureBlock_PetImage)
                    SetPetClass(PetIndexSelected, TextBlock_Class, TextBlock_LVPackageA, TextBlock_LVPackageB)
                    SetPetPurity(PetIndexSelected.Purity, TextBlock_Purity)
                    SetPetTier(PetIndexSelected.Tier, TextBlock_Tier)

                    if(PetIndexSelected.IsBeingUsed = true):
                        TextureBlock_PetImage.SetTint(Red)
                        TextBlock_Class.SetTextColor(Black)
                        TextBlock_LVPackageA.SetTextColor(Black)
                        TextBlock_LVPackageB.SetTextColor(Black)
                        TextBlock_Purity.SetTextColor(Black)
                        TextBlock_Tier.SetTextColor(Black)
                    else:
                        TextureBlock_PetImage.SetTint(White)
                        TextBlock_Class.SetTextColor(White)
                        TextBlock_Purity.SetTextColor(White)
                        TextBlock_Tier.SetTextColor(White)
                else:
                    TextureBlock_PetImage.SetImage(Images.Cross_Icon)
                    TextBlock_Class.SetText(BasicMessage("-"))
                    TextBlock_LVPackageA.SetText(BasicMessage("-"))
                    TextBlock_LVPackageA.SetTextColor(White)
                    TextBlock_LVPackageB.SetText(BasicMessage("-"))
                    TextBlock_LVPackageB.SetTextColor(White)
                    TextBlock_Purity.SetText(BasicMessage("-"))
                    TextBlock_Tier.SetText(BasicMessage("-"))
                    TextureBlock_PetImage.SetTint(White)
                    TextBlock_Class.SetTextColor(White)
                    TextBlock_Purity.SetTextColor(White)
                    TextBlock_Tier.SetTextColor(White)
                    #Pet not found!
                

    Handler_PetSelector_ToChallengeSelector(Message : widget_message) : void =
        Remove_Canvas(Message)      
        spawn:
            HexaMinigameManager.Delay_ChallengeCanvas()


    Remove_Canvas(Message : widget_message) : void =
        set UIIsDisplayed = false
        ##Print("Trying to remove UI")
        if:
            Agent:=MaybeAgent?
            Player:=player[Agent]
            PlayerUI:=GetPlayerUI[Player]
            Canvas := MaybeCanvas?
        then:
            PlayerUI.RemoveWidget(Canvas)
            RemoveDelay_OnRemove()

    Remove_CanvasWithoutMessage() : void =
        set UIIsDisplayed = false
        if:
            Agent:=MaybeAgent?
            Player:=player[Agent]
            PlayerUI:=GetPlayerUI[Player]
            Canvas := MaybeCanvas?
        then:
            PlayerUI.RemoveWidget(Canvas)
            RemoveDelay_OnRemove()

    RemoveDelay_OnAdd() : void =
        SelectButton.SetText(BasicMessage("Select"))

    RemoveDelay_OnRemove() : void =
        SelectButton.SetText(BasicMessage("Select "))
  
    Delay_Canvas()<suspends> : void =
        Sleep(0.1)
        if(not UIIsDisplayed?):
            set UIIsDisplayed = true
            set ConfirmationToReset = 0
            var Index : int = 0 
            if:
                Agent:=MaybeAgent?
                PlayerUI:=GetPlayerUI[player[Agent]]
                Canvas := MaybeCanvas? 
            then:
                ##Print("Creating canvas for the first time")
                PlayerUI.AddWidget(Canvas,player_ui_slot{ZOrder:=0, InputMode := ui_input_mode.All})
                RemoveDelay_OnAdd()
                RefreshAll()
   

    MakeCanvas(): canvas=
        Canvas := canvas:
            Slots:=array:
                #Reset
                canvas_slot:
                    Anchors:=anchors{Minimum:=vector2{X:=0.5,Y:=1.0},Maximum:=vector2{X:=0.5,Y:=1.0}}
                    ZOrder:=0
                    SizeToContent:=true
                    Offsets:=margin{Left:=0.0,Top:=-950.0,Right:=<#1#>00.0,Bottom:=30.0} #PositionX, PositionY, SizeX, SizeY
                    Alignment:=vector2{X:=0.5,Y:=1.0}
                    Widget:=overlay:
                        Slots:=array:
                            overlay_slot:
                                HorizontalAlignment:=horizontal_alignment.Center
                                VerticalAlignment:=vertical_alignment.Center
                                Widget:=texture_block{DefaultImage:=Images.EmptyTexture,DefaultDesiredSize:=vector2{X:=225.0,Y:=125.0}}
                            overlay_slot:
                                HorizontalAlignment:=horizontal_alignment.Fill
                                VerticalAlignment:=vertical_alignment.Fill
                                Widget:=ResetButton
                #Fight
                canvas_slot:
                    Anchors:=anchors{Minimum:=vector2{X:=0.5,Y:=1.0},Maximum:=vector2{X:=0.5,Y:=1.0}}
                    ZOrder:=0
                    SizeToContent:=true
                    Offsets:=margin{Left:=0.0,Top:=-800.0<#-900.0#>,Right:=<#1#>00.0,Bottom:=30.0} #PositionX, PositionY, SizeX, SizeY
                    Alignment:=vector2{X:=0.5,Y:=1.0}
                    Widget:=overlay:
                        Slots:=array:
                            overlay_slot:
                                HorizontalAlignment:=horizontal_alignment.Center
                                VerticalAlignment:=vertical_alignment.Center
                                Widget:=texture_block{DefaultImage:=Images.EmptyTexture,DefaultDesiredSize:=vector2{X:=225.0,Y:=125.0}}
                            overlay_slot:
                                HorizontalAlignment:=horizontal_alignment.Fill
                                VerticalAlignment:=vertical_alignment.Fill
                                Widget:=FightButton
                canvas_slot:
                    Anchors:=anchors{Minimum:=vector2{X:=0.5,Y:=1.0},Maximum:=vector2{X:=0.5,Y:=1.0}}
                    ZOrder:=0
                    SizeToContent:=true
                    Offsets:=margin{Left:=0.0,Top:=-20.0,Right:=100.0,Bottom:=30.0} #PositionX, PositionY, SizeX, SizeY
                    Alignment:=vector2{X:=0.5,Y:=1.0}
                    Widget:=overlay:
                        Slots:=array:
                            overlay_slot:
                                HorizontalAlignment:=horizontal_alignment.Fill
                                VerticalAlignment:=vertical_alignment.Fill
                                Widget:=texture_block{DefaultImage:=Images.EmptyTexture,DefaultDesiredSize:=vector2{X:=BigCloseButtonSizeX-100.0,Y:=70.0}}
                            overlay_slot:
                                HorizontalAlignment:=horizontal_alignment.Fill
                                VerticalAlignment:=vertical_alignment.Fill
                                Widget:=CancelButton
                                
                canvas_slot:
                    Anchors:=anchors{Minimum:=vector2{X:=0.5,Y:=1.0},Maximum:=vector2{X:=0.5,Y:=1.0}}
                    ZOrder:=0
                    SizeToContent:=true
                    Offsets:=margin{Left:=0.0,Top:=-130.0,Right:=0.0,Bottom:=0.0} #PositionX, PositionY, SizeX, SizeY
                    Alignment:=vector2{X:=0.5,Y:=1.0}
                    Widget:=overlay:
                        Slots:=array:
                            overlay_slot:
                                HorizontalAlignment:=horizontal_alignment.Fill
                                VerticalAlignment:=vertical_alignment.Fill
                                Widget:=stack_box:
                                    Orientation:=orientation.Horizontal
                                        Slots:=array:
                                            stack_box_slot: #Slot 0
                                                Padding:=margin{Right:=SelectorBoxPadding,Left:=SelectorBoxPadding}
                                                HorizontalAlignment:=horizontal_alignment.Center
                                                VerticalAlignment:=vertical_alignment.Bottom
                                                Widget:=stack_box:  
                                                    Orientation:=orientation.Vertical
                                                        Slots:=array:
                                                            stack_box_slot: 
                                                                Padding:=margin{Top:=SelectorBoxPadding,Bottom:=SelectorBoxPadding}
                                                                Widget:=overlay:
                                                                    Slots:=array:
                                                                        overlay_slot:
                                                                            HorizontalAlignment:=horizontal_alignment.Fill
                                                                            VerticalAlignment:=vertical_alignment.Fill
                                                                            Widget:=texture_block{DefaultTint:=Black,DefaultImage:=Images.Img_TintedBackgroundHighOpacity}
                                                                        overlay_slot:
                                                                            HorizontalAlignment:=horizontal_alignment.Fill
                                                                            VerticalAlignment:=vertical_alignment.Fill
                                                                            Padding:=margin{Left:=SelectorBoxPadding,Right:=SelectorBoxPadding,Top:=SelectorBoxPadding,Bottom:=SelectorBoxPadding}
                                                                            Widget:=texture_block{DefaultTint:=Black,DefaultImage:=Images.Img_TintedBackgroundHighOpacity}
                                                                        overlay_slot:
                                                                            HorizontalAlignment:=horizontal_alignment.Center
                                                                            VerticalAlignment:=vertical_alignment.Center
                                                                            Widget:=texture_block{DefaultImage:=Images.EmptyTexture,DefaultDesiredSize:=vector2{X:=258.0,Y:=300.0}}
                                                                        overlay_slot:
                                                                            HorizontalAlignment:=horizontal_alignment.Center
                                                                            VerticalAlignment:=vertical_alignment.Center
                                                                            Widget:=stack_box:
                                                                                Orientation:=orientation.Vertical
                                                                                Slots:=array:
                                                                                    stack_box_slot: 
                                                                                        Padding:=margin{Left:=SelectorBoxLinePadding,Right:=SelectorBoxLinePadding,Top:=SelectorBoxLinePadding,Bottom:=SelectorBoxLinePadding}
                                                                                        HorizontalAlignment:=horizontal_alignment.Center
                                                                                        VerticalAlignment:=vertical_alignment.Center
                                                                                        Widget:=TextureBlock_PetImages_Box0
                                                                                    stack_box_slot: 
                                                                                        Padding:=margin{Left:=SelectorBoxLinePadding,Right:=SelectorBoxLinePadding,Top:=SelectorBoxLinePadding,Bottom:=SelectorBoxLinePadding}
                                                                                        HorizontalAlignment:=horizontal_alignment.Center
                                                                                        VerticalAlignment:=vertical_alignment.Center
                                                                                        Widget:=stack_box:
                                                                                            Orientation:=orientation.Horizontal
                                                                                            Slots:=array:
                                                                                                stack_box_slot: 
                                                                                                    Padding:=margin{Left:=2.5,Right:=2.5}
                                                                                                    Widget:=TextBlock_Classes_Box0
                                                                                                stack_box_slot: 
                                                                                                    Padding:=margin{Left:=2.5, Right:=1.0}
                                                                                                    Widget:=TextBlock_PackA_Box0
                                                                                                stack_box_slot: 
                                                                                                    Padding:=margin{Right:=2.5}
                                                                                                    Widget:=TextBlock_PackB_Box0
                                                                                    stack_box_slot: 
                                                                                        Padding:=margin{Left:=SelectorBoxLinePadding,Right:=SelectorBoxLinePadding,Top:=SelectorBoxLinePadding,Bottom:=SelectorBoxLinePadding}
                                                                                        HorizontalAlignment:=horizontal_alignment.Center
                                                                                        VerticalAlignment:=vertical_alignment.Center
                                                                                        Widget:=TextBlock_Purity_Box0
                                                                                    stack_box_slot: 
                                                                                        Padding:=margin{Left:=SelectorBoxLinePadding,Right:=SelectorBoxLinePadding,Top:=SelectorBoxLinePadding,Bottom:=SelectorBoxLinePadding}
                                                                                        HorizontalAlignment:=horizontal_alignment.Center
                                                                                        VerticalAlignment:=vertical_alignment.Center
                                                                                        Widget:=TextBlock_Tier_Box0
                                                            stack_box_slot: 
                                                                Widget:=overlay:
                                                                    Slots:=array:
                                                                        overlay_slot:
                                                                            HorizontalAlignment:=horizontal_alignment.Fill
                                                                            VerticalAlignment:=vertical_alignment.Fill
                                                                            Widget:=texture_block{DefaultImage:=Images.EmptyTexture,DefaultDesiredSize:=QuietButton_PetSelectorDesiredSizeV2}
                                            stack_box_slot: #Slot 1
                                                Padding:=margin{Right:=SelectorBoxPadding,Left:=SelectorBoxPadding}
                                                HorizontalAlignment:=horizontal_alignment.Center
                                                VerticalAlignment:=vertical_alignment.Bottom
                                                Widget:=stack_box:  
                                                    Orientation:=orientation.Vertical
                                                        Slots:=array:
                                                            stack_box_slot: 
                                                                Padding:=margin{Top:=SelectorBoxPadding,Bottom:=SelectorBoxPadding}
                                                                Widget:=overlay:
                                                                    Slots:=array:
                                                                        overlay_slot:
                                                                            HorizontalAlignment:=horizontal_alignment.Fill
                                                                            VerticalAlignment:=vertical_alignment.Fill
                                                                            Widget:=texture_block{DefaultTint:=Black,DefaultImage:=Images.Img_TintedBackgroundHighOpacity}
                                                                        overlay_slot:
                                                                            HorizontalAlignment:=horizontal_alignment.Fill
                                                                            VerticalAlignment:=vertical_alignment.Fill
                                                                            Padding:=margin{Left:=SelectorBoxPadding,Right:=SelectorBoxPadding,Top:=SelectorBoxPadding,Bottom:=SelectorBoxPadding}
                                                                            Widget:=texture_block{DefaultTint:=Black,DefaultImage:=Images.Img_TintedBackgroundHighOpacity}
                                                                        overlay_slot:
                                                                            HorizontalAlignment:=horizontal_alignment.Center
                                                                            VerticalAlignment:=vertical_alignment.Center
                                                                            Widget:=texture_block{DefaultImage:=Images.EmptyTexture,DefaultDesiredSize:=vector2{X:=258.0,Y:=300.0}}
                                                                        overlay_slot:
                                                                            HorizontalAlignment:=horizontal_alignment.Center
                                                                            VerticalAlignment:=vertical_alignment.Center
                                                                            Widget:=stack_box:
                                                                                Orientation:=orientation.Vertical
                                                                                Slots:=array:
                                                                                    stack_box_slot: 
                                                                                        Padding:=margin{Left:=SelectorBoxLinePadding,Right:=SelectorBoxLinePadding,Top:=SelectorBoxLinePadding,Bottom:=SelectorBoxLinePadding}
                                                                                        HorizontalAlignment:=horizontal_alignment.Center
                                                                                        VerticalAlignment:=vertical_alignment.Center
                                                                                        Widget:=TextureBlock_PetImages_Box1
                                                                                    stack_box_slot: 
                                                                                        Padding:=margin{Left:=SelectorBoxLinePadding,Right:=SelectorBoxLinePadding,Top:=SelectorBoxLinePadding,Bottom:=SelectorBoxLinePadding}
                                                                                        HorizontalAlignment:=horizontal_alignment.Center
                                                                                        VerticalAlignment:=vertical_alignment.Center
                                                                                        Widget:=stack_box:
                                                                                            Orientation:=orientation.Horizontal
                                                                                            Slots:=array:
                                                                                                stack_box_slot: 
                                                                                                    Padding:=margin{Left:=2.5,Right:=2.5}
                                                                                                    Widget:=TextBlock_Classes_Box1
                                                                                                stack_box_slot: 
                                                                                                    Padding:=margin{Left:=2.5, Right:=1.0}
                                                                                                    Widget:=TextBlock_PackA_Box1
                                                                                                stack_box_slot: 
                                                                                                    Padding:=margin{Right:=2.5}
                                                                                                    Widget:=TextBlock_PackB_Box1
                                                                                        
                                                                                    stack_box_slot: 
                                                                                        Padding:=margin{Left:=SelectorBoxLinePadding,Right:=SelectorBoxLinePadding,Top:=SelectorBoxLinePadding,Bottom:=SelectorBoxLinePadding}
                                                                                        HorizontalAlignment:=horizontal_alignment.Center
                                                                                        VerticalAlignment:=vertical_alignment.Center
                                                                                        Widget:=TextBlock_Purity_Box1
                                                                                    stack_box_slot: 
                                                                                        Padding:=margin{Left:=SelectorBoxLinePadding,Right:=SelectorBoxLinePadding,Top:=SelectorBoxLinePadding,Bottom:=SelectorBoxLinePadding}
                                                                                        HorizontalAlignment:=horizontal_alignment.Center
                                                                                        VerticalAlignment:=vertical_alignment.Center
                                                                                        Widget:=TextBlock_Tier_Box1
                                                            stack_box_slot: 
                                                                Widget:=overlay:
                                                                    Slots:=array:
                                                                        overlay_slot:
                                                                            HorizontalAlignment:=horizontal_alignment.Fill
                                                                            VerticalAlignment:=vertical_alignment.Fill
                                                                            Widget:=texture_block{DefaultImage:=Images.EmptyTexture,DefaultDesiredSize:=QuietButton_PetSelectorDesiredSizeV2}
                                                                        overlay_slot:
                                                                            HorizontalAlignment:=horizontal_alignment.Fill
                                                                            Widget:=NavigateLeftButton
                                            stack_box_slot: #Slot 3
                                                Padding:=margin{Right:=SelectorBoxPadding,Left:=SelectorBoxPadding}
                                                HorizontalAlignment:=horizontal_alignment.Center
                                                VerticalAlignment:=vertical_alignment.Bottom
                                                Widget:=stack_box:  
                                                    Orientation:=orientation.Vertical
                                                        Slots:=array:
                                                            stack_box_slot: 
                                                                Padding:=margin{Top:=SelectorBoxPadding,Bottom:=SelectorBoxPadding}
                                                                Widget:=overlay:
                                                                    Slots:=array:
                                                                        overlay_slot:
                                                                            HorizontalAlignment:=horizontal_alignment.Fill
                                                                            VerticalAlignment:=vertical_alignment.Fill
                                                                            Widget:=texture_block{DefaultTint:=Black,DefaultImage:=Images.Img_TintedBackgroundHighOpacity}
                                                                        overlay_slot:
                                                                            HorizontalAlignment:=horizontal_alignment.Fill
                                                                            VerticalAlignment:=vertical_alignment.Fill
                                                                            Padding:=margin{Left:=SelectorBoxPadding,Right:=SelectorBoxPadding,Top:=SelectorBoxPadding,Bottom:=SelectorBoxPadding}
                                                                            Widget:=texture_block{DefaultTint:=Black,DefaultImage:=Images.Img_TintedBackgroundHighOpacity}
                                                                        overlay_slot:
                                                                            HorizontalAlignment:=horizontal_alignment.Center
                                                                            VerticalAlignment:=vertical_alignment.Center
                                                                            Widget:=texture_block{DefaultImage:=Images.EmptyTexture,DefaultDesiredSize:=vector2{X:=500.0,Y:=560.0}}
                                                                        overlay_slot:
                                                                            HorizontalAlignment:=horizontal_alignment.Center
                                                                            VerticalAlignment:=vertical_alignment.Center
                                                                            Widget:=stack_box:
                                                                                Orientation:=orientation.Vertical
                                                                                Slots:=array:
                                                                                    stack_box_slot: 
                                                                                        Padding:=margin{Left:=SelectorBoxLinePadding,Right:=SelectorBoxLinePadding,Top:=SelectorBoxLinePadding,Bottom:=SelectorBoxLinePadding}
                                                                                        HorizontalAlignment:=horizontal_alignment.Center
                                                                                        VerticalAlignment:=vertical_alignment.Center
                                                                                        Widget:=Selected_TextureBlock_PetImages
                                                                                    stack_box_slot: 
                                                                                        Padding:=margin{Left:=SelectorBoxLinePadding,Right:=SelectorBoxLinePadding,Top:=SelectorBoxLinePadding,Bottom:=SelectorBoxLinePadding}
                                                                                        HorizontalAlignment:=horizontal_alignment.Center
                                                                                        VerticalAlignment:=vertical_alignment.Center
                                                                                        Widget:=stack_box:
                                                                                            Orientation:=orientation.Horizontal
                                                                                            Slots:=array:
                                                                                                stack_box_slot: 
                                                                                                    Padding:=margin{Left:=2.5,Right:=2.5}
                                                                                                    Widget:=Selected_TextBlock_Class
                                                                                                stack_box_slot: 
                                                                                                    Padding:=margin{Left:=2.5, Right:=1.0}
                                                                                                    Widget:=Selected_TextBlock_LVPackageA
                                                                                                stack_box_slot: 
                                                                                                    Padding:=margin{Right:=2.5}
                                                                                                    Widget:=Selected_TextBlock_LVPackageB
                                                                                    stack_box_slot: 
                                                                                        Padding:=margin{Left:=SelectorBoxLinePadding,Right:=SelectorBoxLinePadding,Top:=SelectorBoxLinePadding,Bottom:=SelectorBoxLinePadding}
                                                                                        HorizontalAlignment:=horizontal_alignment.Center
                                                                                        VerticalAlignment:=vertical_alignment.Center
                                                                                        Widget:=Selected_TextBlock_Purity
                                                                                    stack_box_slot: 
                                                                                        HorizontalAlignment:=horizontal_alignment.Fill
                                                                                        VerticalAlignment:=vertical_alignment.Center
                                                                                        Widget:=texture_block{DefaultDesiredSize:= vector2{X:=1.0,Y:=3.0},DefaultImage:=Images.Img_HorizontalLine}
                                                                                    stack_box_slot: 
                                                                                        HorizontalAlignment:=horizontal_alignment.Fill
                                                                                        VerticalAlignment:=vertical_alignment.Fill
                                                                                        Widget:=stack_box:
                                                                                            Orientation:=orientation.Horizontal
                                                                                            Slots:=array:
                                                                                                stack_box_slot: #Column 1
                                                                                                    HorizontalAlignment:=horizontal_alignment.Fill
                                                                                                    VerticalAlignment:=vertical_alignment.Fill
                                                                                                    Widget:=stack_box:
                                                                                                        Orientation:=orientation.Vertical
                                                                                                        Slots:=array:
                                                                                                            stack_box_slot: 
                                                                                                                Padding:=margin{Left:=SelectorBoxLinePadding,Right:=SelectorBoxLinePadding,Top:=SelectorBoxLinePadding,Bottom:=SelectorBoxLinePadding}
                                                                                                                HorizontalAlignment:=horizontal_alignment.Left
                                                                                                                VerticalAlignment:=vertical_alignment.Center
                                                                                                                Widget:=text_block{DefaultTextColor := White, DefaultText := BasicMessage("Max Health:")}
                                                                                                            stack_box_slot: 
                                                                                                                Padding:=margin{Left:=SelectorBoxLinePadding,Right:=SelectorBoxLinePadding,Top:=SelectorBoxLinePadding,Bottom:=SelectorBoxLinePadding}
                                                                                                                HorizontalAlignment:=horizontal_alignment.Left
                                                                                                                VerticalAlignment:=vertical_alignment.Center
                                                                                                                Widget:=text_block{DefaultTextColor := White, DefaultText := BasicMessage("Damage:")}
                                                                                                            stack_box_slot: 
                                                                                                                Padding:=margin{Left:=SelectorBoxLinePadding,Right:=SelectorBoxLinePadding,Top:=SelectorBoxLinePadding,Bottom:=SelectorBoxLinePadding}
                                                                                                                HorizontalAlignment:=horizontal_alignment.Left
                                                                                                                VerticalAlignment:=vertical_alignment.Center
                                                                                                                Widget:=text_block{DefaultTextColor := White, DefaultText := BasicMessage("Healing perk:")}
                                                                                                            stack_box_slot: 
                                                                                                                Padding:=margin{Left:=SelectorBoxLinePadding,Right:=SelectorBoxLinePadding,Top:=SelectorBoxLinePadding,Bottom:=SelectorBoxLinePadding}
                                                                                                                HorizontalAlignment:=horizontal_alignment.Left
                                                                                                                VerticalAlignment:=vertical_alignment.Center
                                                                                                                Widget:=text_block{DefaultTextColor := White, DefaultText := BasicMessage("Attack speed:")}
                                                                                                            stack_box_slot: 
                                                                                                                Padding:=margin{Left:=SelectorBoxLinePadding,Right:=SelectorBoxLinePadding,Top:=SelectorBoxLinePadding,Bottom:=SelectorBoxLinePadding}
                                                                                                                HorizontalAlignment:=horizontal_alignment.Left
                                                                                                                VerticalAlignment:=vertical_alignment.Center
                                                                                                                Widget:=text_block{DefaultTextColor := White, DefaultText := BasicMessage("Speed:")}
                                                                                                            stack_box_slot: 
                                                                                                                Padding:=margin{Left:=SelectorBoxLinePadding,Right:=SelectorBoxLinePadding,Top:=SelectorBoxLinePadding,Bottom:=SelectorBoxLinePadding}
                                                                                                                HorizontalAlignment:=horizontal_alignment.Left
                                                                                                                VerticalAlignment:=vertical_alignment.Center
                                                                                                                Widget:=text_block{DefaultTextColor := White, DefaultText := BasicMessage("Range:")}
                                                                                                stack_box_slot: #Column 2
                                                                                                    HorizontalAlignment:=horizontal_alignment.Fill
                                                                                                    VerticalAlignment:=vertical_alignment.Fill
                                                                                                    Widget:=stack_box:
                                                                                                        Orientation:=orientation.Vertical
                                                                                                        Slots:=array:
                                                                                                            stack_box_slot: 
                                                                                                                Padding:=margin{Left:=SelectorBoxLinePadding,Right:=SelectorBoxLinePadding,Top:=SelectorBoxLinePadding,Bottom:=SelectorBoxLinePadding}
                                                                                                                HorizontalAlignment:=horizontal_alignment.Left
                                                                                                                VerticalAlignment:=vertical_alignment.Center
                                                                                                                Widget:=Selected_TextBlock_Health
                                                                                                            stack_box_slot: 
                                                                                                                Padding:=margin{Left:=SelectorBoxLinePadding,Right:=SelectorBoxLinePadding,Top:=SelectorBoxLinePadding,Bottom:=SelectorBoxLinePadding}
                                                                                                                HorizontalAlignment:=horizontal_alignment.Left
                                                                                                                VerticalAlignment:=vertical_alignment.Center
                                                                                                                Widget:=Selected_TextBlock_Damage
                                                                                                            stack_box_slot: 
                                                                                                                Padding:=margin{Left:=SelectorBoxLinePadding,Right:=SelectorBoxLinePadding,Top:=SelectorBoxLinePadding,Bottom:=SelectorBoxLinePadding}
                                                                                                                HorizontalAlignment:=horizontal_alignment.Left
                                                                                                                VerticalAlignment:=vertical_alignment.Center
                                                                                                                Widget:=Selected_TextBlock_Healing
                                                                                                            stack_box_slot: 
                                                                                                                Padding:=margin{Left:=SelectorBoxLinePadding,Right:=SelectorBoxLinePadding,Top:=SelectorBoxLinePadding,Bottom:=SelectorBoxLinePadding}
                                                                                                                HorizontalAlignment:=horizontal_alignment.Left
                                                                                                                VerticalAlignment:=vertical_alignment.Center
                                                                                                                Widget:=Selected_TextBlock_AttackSpeed
                                                                                                            stack_box_slot: 
                                                                                                                Padding:=margin{Left:=SelectorBoxLinePadding,Right:=SelectorBoxLinePadding,Top:=SelectorBoxLinePadding,Bottom:=SelectorBoxLinePadding}
                                                                                                                HorizontalAlignment:=horizontal_alignment.Left
                                                                                                                VerticalAlignment:=vertical_alignment.Center
                                                                                                                Widget:=Selected_TextBlock_Speed
                                                                                                            stack_box_slot: 
                                                                                                                Padding:=margin{Left:=SelectorBoxLinePadding,Right:=SelectorBoxLinePadding,Top:=SelectorBoxLinePadding,Bottom:=SelectorBoxLinePadding}
                                                                                                                HorizontalAlignment:=horizontal_alignment.Left
                                                                                                                VerticalAlignment:=vertical_alignment.Center
                                                                                                                Widget:=Selected_TextBlock_Range
                                                            stack_box_slot: 
                                                                HorizontalAlignment:=horizontal_alignment.Fill
                                                                VerticalAlignment:=vertical_alignment.Fill
                                                                Widget:=overlay:
                                                                    Slots:=array:
                                                                        overlay_slot:
                                                                            HorizontalAlignment:=horizontal_alignment.Fill
                                                                            VerticalAlignment:=vertical_alignment.Fill
                                                                            Widget:=texture_block{DefaultImage:=Images.EmptyTexture,DefaultDesiredSize:=QuietButton_PetSelectorDesiredSizeV2}
                                                                        overlay_slot:
                                                                            HorizontalAlignment:=horizontal_alignment.Fill
                                                                            VerticalAlignment:=vertical_alignment.Fill
                                                                            Widget:=SelectButton
                                            stack_box_slot: #Slot 4
                                                Padding:=margin{Right:=SelectorBoxPadding,Left:=SelectorBoxPadding}
                                                HorizontalAlignment:=horizontal_alignment.Center
                                                VerticalAlignment:=vertical_alignment.Bottom
                                                Widget:=stack_box:  
                                                    Orientation:=orientation.Vertical
                                                        Slots:=array:
                                                            stack_box_slot: 
                                                                Padding:=margin{Top:=SelectorBoxPadding,Bottom:=SelectorBoxPadding}
                                                                Widget:=overlay:
                                                                    Slots:=array:
                                                                        overlay_slot:
                                                                            HorizontalAlignment:=horizontal_alignment.Fill
                                                                            VerticalAlignment:=vertical_alignment.Fill
                                                                            Widget:=texture_block{DefaultTint:=Black,DefaultImage:=Images.Img_TintedBackgroundHighOpacity}
                                                                        overlay_slot:
                                                                            HorizontalAlignment:=horizontal_alignment.Fill
                                                                            VerticalAlignment:=vertical_alignment.Fill
                                                                            Padding:=margin{Left:=SelectorBoxPadding,Right:=SelectorBoxPadding,Top:=SelectorBoxPadding,Bottom:=SelectorBoxPadding}
                                                                            Widget:=texture_block{DefaultTint:=Black,DefaultImage:=Images.Img_TintedBackgroundHighOpacity}
                                                                        overlay_slot:
                                                                            HorizontalAlignment:=horizontal_alignment.Center
                                                                            VerticalAlignment:=vertical_alignment.Center
                                                                            Widget:=texture_block{DefaultImage:=Images.EmptyTexture,DefaultDesiredSize:=vector2{X:=258.0,Y:=300.0}}
                                                                        overlay_slot:
                                                                            HorizontalAlignment:=horizontal_alignment.Center
                                                                            VerticalAlignment:=vertical_alignment.Center
                                                                            Widget:=stack_box:
                                                                                Orientation:=orientation.Vertical
                                                                                Slots:=array:
                                                                                    stack_box_slot: 
                                                                                        Padding:=margin{Left:=SelectorBoxLinePadding,Right:=SelectorBoxLinePadding,Top:=SelectorBoxLinePadding,Bottom:=SelectorBoxLinePadding}
                                                                                        HorizontalAlignment:=horizontal_alignment.Center
                                                                                        VerticalAlignment:=vertical_alignment.Center
                                                                                        Widget:=TextureBlock_PetImages_Box2
                                                                                    stack_box_slot: 
                                                                                        Padding:=margin{Left:=SelectorBoxLinePadding,Right:=SelectorBoxLinePadding,Top:=SelectorBoxLinePadding,Bottom:=SelectorBoxLinePadding}
                                                                                        HorizontalAlignment:=horizontal_alignment.Center
                                                                                        VerticalAlignment:=vertical_alignment.Center
                                                                                        Widget:=stack_box:
                                                                                            Orientation:=orientation.Horizontal
                                                                                            Slots:=array:
                                                                                                stack_box_slot: 
                                                                                                    Padding:=margin{Left:=2.5,Right:=2.5}
                                                                                                    Widget:=TextBlock_Classes_Box2
                                                                                                stack_box_slot: 
                                                                                                    Padding:=margin{Left:=2.5, Right:=1.0}
                                                                                                    Widget:=TextBlock_PackA_Box2
                                                                                                stack_box_slot: 
                                                                                                    Padding:=margin{Right:=2.5}
                                                                                                    Widget:=TextBlock_PackB_Box2
                                                                                    stack_box_slot: 
                                                                                        Padding:=margin{Left:=SelectorBoxLinePadding,Right:=SelectorBoxLinePadding,Top:=SelectorBoxLinePadding,Bottom:=SelectorBoxLinePadding}
                                                                                        HorizontalAlignment:=horizontal_alignment.Center
                                                                                        VerticalAlignment:=vertical_alignment.Center
                                                                                        Widget:=TextBlock_Purity_Box2
                                                                                    stack_box_slot: 
                                                                                        Padding:=margin{Left:=SelectorBoxLinePadding,Right:=SelectorBoxLinePadding,Top:=SelectorBoxLinePadding,Bottom:=SelectorBoxLinePadding}
                                                                                        HorizontalAlignment:=horizontal_alignment.Center
                                                                                        VerticalAlignment:=vertical_alignment.Center
                                                                                        Widget:=TextBlock_Tier_Box2
                                                            stack_box_slot: 
                                                                Widget:=overlay:
                                                                    Slots:=array:
                                                                        overlay_slot:
                                                                            HorizontalAlignment:=horizontal_alignment.Fill
                                                                            VerticalAlignment:=vertical_alignment.Fill
                                                                            Widget:=texture_block{DefaultImage:=Images.EmptyTexture,DefaultDesiredSize:=QuietButton_PetSelectorDesiredSizeV2}
                                                                        overlay_slot:
                                                                            HorizontalAlignment:=horizontal_alignment.Fill
                                                                            Widget:=NavigateRightButton

                                            stack_box_slot: #Slot 5
                                                Padding:=margin{Right:=SelectorBoxPadding,Left:=SelectorBoxPadding}
                                                HorizontalAlignment:=horizontal_alignment.Center
                                                VerticalAlignment:=vertical_alignment.Bottom
                                                Widget:=stack_box:  
                                                    Orientation:=orientation.Vertical
                                                        Slots:=array:
                                                            stack_box_slot: 
                                                                Padding:=margin{Top:=SelectorBoxPadding,Bottom:=SelectorBoxPadding}
                                                                Widget:=overlay:
                                                                    Slots:=array:
                                                                        overlay_slot:
                                                                            HorizontalAlignment:=horizontal_alignment.Fill
                                                                            VerticalAlignment:=vertical_alignment.Fill
                                                                            Widget:=texture_block{DefaultTint:=Black,DefaultImage:=Images.Img_TintedBackgroundHighOpacity}
                                                                        overlay_slot:
                                                                            HorizontalAlignment:=horizontal_alignment.Fill
                                                                            VerticalAlignment:=vertical_alignment.Fill
                                                                            Padding:=margin{Left:=SelectorBoxPadding,Right:=SelectorBoxPadding,Top:=SelectorBoxPadding,Bottom:=SelectorBoxPadding}
                                                                            Widget:=texture_block{DefaultTint:=Black,DefaultImage:=Images.Img_TintedBackgroundHighOpacity}
                                                                        overlay_slot:
                                                                            HorizontalAlignment:=horizontal_alignment.Center
                                                                            VerticalAlignment:=vertical_alignment.Center
                                                                            Widget:=texture_block{DefaultImage:=Images.EmptyTexture,DefaultDesiredSize:=vector2{X:=258.0,Y:=300.0}}
                                                                        overlay_slot:
                                                                            HorizontalAlignment:=horizontal_alignment.Center
                                                                            VerticalAlignment:=vertical_alignment.Center
                                                                            Widget:=stack_box:
                                                                                Orientation:=orientation.Vertical
                                                                                Slots:=array:
                                                                                    stack_box_slot: 
                                                                                        Padding:=margin{Left:=SelectorBoxLinePadding,Right:=SelectorBoxLinePadding,Top:=SelectorBoxLinePadding,Bottom:=SelectorBoxLinePadding}
                                                                                        HorizontalAlignment:=horizontal_alignment.Center
                                                                                        VerticalAlignment:=vertical_alignment.Center
                                                                                        Widget:=TextureBlock_PetImages_Box3
                                                                                    stack_box_slot: 
                                                                                        Padding:=margin{Left:=SelectorBoxLinePadding,Right:=SelectorBoxLinePadding,Top:=SelectorBoxLinePadding,Bottom:=SelectorBoxLinePadding}
                                                                                        HorizontalAlignment:=horizontal_alignment.Center
                                                                                        VerticalAlignment:=vertical_alignment.Center
                                                                                        Widget:=stack_box:
                                                                                            Orientation:=orientation.Horizontal
                                                                                            Slots:=array:
                                                                                                stack_box_slot: 
                                                                                                    Padding:=margin{Left:=2.5,Right:=2.5}
                                                                                                    Widget:=TextBlock_Classes_Box3
                                                                                                stack_box_slot: 
                                                                                                    Padding:=margin{Left:=2.5, Right:=1.0}
                                                                                                    Widget:=TextBlock_PackA_Box3
                                                                                                stack_box_slot: 
                                                                                                    Padding:=margin{Right:=2.5}
                                                                                                    Widget:=TextBlock_PackB_Box3
                                                                                    stack_box_slot: 
                                                                                        Padding:=margin{Left:=SelectorBoxLinePadding,Right:=SelectorBoxLinePadding,Top:=SelectorBoxLinePadding,Bottom:=SelectorBoxLinePadding}
                                                                                        HorizontalAlignment:=horizontal_alignment.Center
                                                                                        VerticalAlignment:=vertical_alignment.Center
                                                                                        Widget:=TextBlock_Purity_Box3
                                                                                    stack_box_slot: 
                                                                                        Padding:=margin{Left:=SelectorBoxLinePadding,Right:=SelectorBoxLinePadding,Top:=SelectorBoxLinePadding,Bottom:=SelectorBoxLinePadding}
                                                                                        HorizontalAlignment:=horizontal_alignment.Center
                                                                                        VerticalAlignment:=vertical_alignment.Center
                                                                                        Widget:=TextBlock_Tier_Box3
                                                            stack_box_slot: 
                                                                Widget:=overlay:
                                                                    Slots:=array:
                                                                        overlay_slot:
                                                                            HorizontalAlignment:=horizontal_alignment.Fill
                                                                            VerticalAlignment:=vertical_alignment.Fill
                                                                            Widget:=texture_block{DefaultImage:=Images.EmptyTexture,DefaultDesiredSize:=QuietButton_PetSelectorDesiredSizeV2}
                                                           
MakeHexMiniGameSelectorUI<constructor>(
    InHexaMinigameManager : hexa_minigame_manager)<transacts> := hexa_minigame_selector_ui :
        HexaMinigameManager := InHexaMinigameManager
                                                                                                                