# fill this file with your verse-script
using { /UnrealEngine.com/Temporary/UI }
using { /UnrealEngine.com/Temporary/SpatialMath }
using { /Fortnite.com/UI }
using { /Verse.org/Simulation }
using { /Verse.org/Colors/NamedColors }

# A Verse-authored creative device that can be placed in a level
hexa_minigame_ui := class():

    #Class settings
    var MaybeAgent : ?agent = false
    var TotalGameTime : int = 59
    HexaMinigameManager : hexa_minigame_manager

    #Close Button overlay
    CloseButtonOverlay : overlay = overlay{}

    #Reamining fight time
    RemainingFightTime : text_block = text_block{DefaultTextColor:=Green,DefaultShadowOffset:=ShadowOffset, DefaultShadowOpacity := 1.0}

    #Slot 1 Player
    EmptyImage_Player_Slot1 : texture_block = texture_block{DefaultImage:=Images.EmptyTexture, DefaultDesiredSize := vector2{X:=SizeSlotX_PetHexaUI,Y:=SizeSlotY_PetHexaUI}}
    HealthLine_Player_Slot1 : texture_block = texture_block{DefaultTint:=Green, DefaultImage:=Images.Img_TintedBackgroundHighOpacity, DefaultDesiredSize := vector2{X:=0.0,Y:=0.0}}
    PetImage_Player_Slot1 : texture_block = texture_block{DefaultImage:=Images.Img_LineBackground, DefaultDesiredSize := vector2{X:=SizeImageX_PetHexaUI,Y:=SizeImageY_PetHexaUI}}
    Class_Player_Slot1 : text_block = text_block{DefaultTextColor:=White,DefaultShadowOffset:=ShadowOffset, DefaultShadowOpacity := 1.0}
    PackA_Player_Box1 : text_block = text_block{DefaultTextColor := White,DefaultShadowOffset:=ShadowOffset, DefaultShadowOpacity := 1.0}
    PackB_Player_Box1 : text_block = text_block{DefaultTextColor := White,DefaultShadowOffset:=ShadowOffset, DefaultShadowOpacity := 1.0}
    Purity_Player_Slot1 : text_block = text_block{DefaultTextColor:=White,DefaultShadowOffset:=ShadowOffset, DefaultShadowOpacity := 1.0}
    Health_Player_Slot1 : text_block = text_block{DefaultTextColor:=White,DefaultShadowOffset:=ShadowOffset, DefaultShadowOpacity := 1.0}

    #Slot 2 Player
    EmptyImage_Player_Slot2 : texture_block = texture_block{DefaultImage:=Images.EmptyTexture, DefaultDesiredSize := vector2{X:=SizeSlotX_PetHexaUI,Y:=SizeSlotY_PetHexaUI}}
    HealthLine_Player_Slot2 : texture_block = texture_block{DefaultTint:=Green,DefaultImage:=Images.Img_TintedBackgroundHighOpacity, DefaultDesiredSize := vector2{X:=0.0,Y:=0.0}}
    PetImage_Player_Slot2 : texture_block = texture_block{DefaultImage:=Images.Img_LineBackground, DefaultDesiredSize := vector2{X:=SizeImageX_PetHexaUI,Y:=SizeImageY_PetHexaUI}}
    Class_Player_Slot2 : text_block = text_block{DefaultTextColor:=White, DefaultShadowOffset:=ShadowOffset, DefaultShadowOpacity := 1.0}
    PackA_Player_Box2 : text_block = text_block{DefaultTextColor := White,DefaultShadowOffset:=ShadowOffset, DefaultShadowOpacity := 1.0}
    PackB_Player_Box2 : text_block = text_block{DefaultTextColor := White,DefaultShadowOffset:=ShadowOffset, DefaultShadowOpacity := 1.0}
    Purity_Player_Slot2 : text_block = text_block{DefaultTextColor:=White,DefaultShadowOffset:=ShadowOffset, DefaultShadowOpacity := 1.0}
    Health_Player_Slot2 : text_block = text_block{DefaultTextColor:=White,DefaultShadowOffset:=ShadowOffset, DefaultShadowOpacity := 1.0}

    #Slot 3 Player
    EmptyImage_Player_Slot3 : texture_block = texture_block{DefaultImage:=Images.EmptyTexture, DefaultDesiredSize := vector2{X:=SizeSlotX_PetHexaUI,Y:=SizeSlotY_PetHexaUI}}
    HealthLine_Player_Slot3 : texture_block = texture_block{DefaultTint:=Green,DefaultImage:=Images.Img_TintedBackgroundHighOpacity, DefaultDesiredSize := vector2{X:=0.0,Y:=0.0}}
    PetImage_Player_Slot3 : texture_block = texture_block{DefaultImage:=Images.Img_LineBackground, DefaultDesiredSize := vector2{X:=SizeImageX_PetHexaUI,Y:=SizeImageY_PetHexaUI}}
    PackA_Player_Box3 : text_block = text_block{DefaultTextColor := White,DefaultShadowOffset:=ShadowOffset, DefaultShadowOpacity := 1.0}
    PackB_Player_Box3 : text_block = text_block{DefaultTextColor := White,DefaultShadowOffset:=ShadowOffset, DefaultShadowOpacity := 1.0}
    Class_Player_Slot3 : text_block = text_block{DefaultTextColor:=White,DefaultShadowOffset:=ShadowOffset, DefaultShadowOpacity := 1.0}
    Purity_Player_Slot3 : text_block = text_block{DefaultTextColor:=White,DefaultShadowOffset:=ShadowOffset, DefaultShadowOpacity := 1.0}
    Health_Player_Slot3 : text_block = text_block{DefaultTextColor:=White,DefaultShadowOffset:=ShadowOffset, DefaultShadowOpacity := 1.0}

    #Slot 4 Player
    EmptyImage_Player_Slot4 : texture_block = texture_block{DefaultImage:=Images.EmptyTexture, DefaultDesiredSize := vector2{X:=SizeSlotX_PetHexaUI,Y:=SizeSlotY_PetHexaUI}}
    HealthLine_Player_Slot4 : texture_block = texture_block{DefaultTint:=Green,DefaultImage:=Images.Img_TintedBackgroundHighOpacity, DefaultDesiredSize := vector2{X:=0.0,Y:=0.0}}
    PetImage_Player_Slot4 : texture_block = texture_block{DefaultImage:=Images.Img_LineBackground, DefaultDesiredSize := vector2{X:=SizeImageX_PetHexaUI,Y:=SizeImageY_PetHexaUI}}
    PackA_Player_Box4 : text_block = text_block{DefaultTextColor := White,DefaultShadowOffset:=ShadowOffset, DefaultShadowOpacity := 1.0}
    PackB_Player_Box4 : text_block = text_block{DefaultTextColor := White,DefaultShadowOffset:=ShadowOffset, DefaultShadowOpacity := 1.0}
    Class_Player_Slot4 : text_block = text_block{DefaultTextColor:=White,DefaultShadowOffset:=ShadowOffset, DefaultShadowOpacity := 1.0}
    Purity_Player_Slot4 : text_block = text_block{DefaultTextColor:=White,DefaultShadowOffset:=ShadowOffset, DefaultShadowOpacity := 1.0}
    Health_Player_Slot4 : text_block = text_block{DefaultTextColor:=White,DefaultShadowOffset:=ShadowOffset, DefaultShadowOpacity := 1.0}

    #Slot 5 Player
    EmptyImage_Player_Slot5 : texture_block = texture_block{DefaultImage:=Images.EmptyTexture, DefaultDesiredSize := vector2{X:=SizeSlotX_PetHexaUI,Y:=SizeSlotY_PetHexaUI}}
    HealthLine_Player_Slot5 : texture_block = texture_block{DefaultTint:=Green,DefaultImage:=Images.Img_TintedBackgroundHighOpacity, DefaultDesiredSize := vector2{X:=0.0,Y:=0.0}}
    PetImage_Player_Slot5 : texture_block = texture_block{DefaultImage:=Images.Img_LineBackground, DefaultDesiredSize := vector2{X:=SizeImageX_PetHexaUI,Y:=SizeImageY_PetHexaUI}}
    PackA_Player_Box5 : text_block = text_block{DefaultTextColor := White,DefaultShadowOffset:=ShadowOffset, DefaultShadowOpacity := 1.0}
    PackB_Player_Box5 : text_block = text_block{DefaultTextColor := White,DefaultShadowOffset:=ShadowOffset, DefaultShadowOpacity := 1.0}
    Class_Player_Slot5 : text_block = text_block{DefaultTextColor:=White,DefaultShadowOffset:=ShadowOffset, DefaultShadowOpacity := 1.0}
    Purity_Player_Slot5 : text_block = text_block{DefaultTextColor:=White,DefaultShadowOffset:=ShadowOffset, DefaultShadowOpacity := 1.0}
    Health_Player_Slot5 : text_block = text_block{DefaultTextColor:=White,DefaultShadowOffset:=ShadowOffset, DefaultShadowOpacity := 1.0}

    #Slot 1 AI
    EmptyImage_AI_Slot1 : texture_block = texture_block{DefaultImage:=Images.EmptyTexture, DefaultDesiredSize := vector2{X:=SizeSlotX_PetHexaUI,Y:=SizeSlotY_PetHexaUI}}
    HealthLine_AI_Slot1 : texture_block = texture_block{DefaultTint:=Red, DefaultImage:=Images.Img_TintedBackgroundHighOpacity, DefaultDesiredSize := vector2{X:=0.0,Y:=0.0}}
    PetImage_AI_Slot1 : texture_block = texture_block{DefaultImage:=Images.Img_LineBackground, DefaultDesiredSize := vector2{X:=SizeImageX_PetHexaUI,Y:=SizeImageY_PetHexaUI}}
    Class_AI_Slot1 : text_block = text_block{DefaultTextColor:=White,DefaultShadowOffset:=ShadowOffset, DefaultShadowOpacity := 1.0}
    PackA_AI_Box1 : text_block = text_block{DefaultTextColor := White,DefaultShadowOffset:=ShadowOffset, DefaultShadowOpacity := 1.0}
    PackB_AI_Box1 : text_block = text_block{DefaultTextColor := White,DefaultShadowOffset:=ShadowOffset, DefaultShadowOpacity := 1.0}
    Purity_AI_Slot1 : text_block = text_block{DefaultTextColor:=White,DefaultShadowOffset:=ShadowOffset, DefaultShadowOpacity := 1.0}
    Health_AI_Slot1 : text_block = text_block{DefaultTextColor:=White,DefaultShadowOffset:=ShadowOffset, DefaultShadowOpacity := 1.0}

    #Slot 2 AI
    EmptyImage_AI_Slot2 : texture_block = texture_block{DefaultImage:=Images.EmptyTexture, DefaultDesiredSize := vector2{X:=SizeSlotX_PetHexaUI,Y:=SizeSlotY_PetHexaUI}}
    HealthLine_AI_Slot2 : texture_block = texture_block{DefaultTint:=Red, DefaultImage:=Images.Img_TintedBackgroundHighOpacity, DefaultDesiredSize := vector2{X:=0.0,Y:=0.0}}
    PetImage_AI_Slot2 : texture_block = texture_block{DefaultImage:=Images.Img_LineBackground, DefaultDesiredSize := vector2{X:=SizeImageX_PetHexaUI,Y:=SizeImageY_PetHexaUI}}
    Class_AI_Slot2 : text_block = text_block{DefaultTextColor:=White,DefaultShadowOffset:=ShadowOffset, DefaultShadowOpacity := 1.0}
    PackA_AI_Box2 : text_block = text_block{DefaultTextColor := White,DefaultShadowOffset:=ShadowOffset, DefaultShadowOpacity := 1.0}
    PackB_AI_Box2 : text_block = text_block{DefaultTextColor := White,DefaultShadowOffset:=ShadowOffset, DefaultShadowOpacity := 1.0}
    Purity_AI_Slot2 : text_block = text_block{DefaultTextColor:=White,DefaultShadowOffset:=ShadowOffset, DefaultShadowOpacity := 1.0}
    Health_AI_Slot2 : text_block = text_block{DefaultTextColor:=White,DefaultShadowOffset:=ShadowOffset, DefaultShadowOpacity := 1.0}

    #Slot 3 AI
    EmptyImage_AI_Slot3 : texture_block = texture_block{DefaultImage:=Images.EmptyTexture, DefaultDesiredSize := vector2{X:=SizeSlotX_PetHexaUI,Y:=SizeSlotY_PetHexaUI}}
    HealthLine_AI_Slot3 : texture_block = texture_block{DefaultTint:=Red, DefaultImage:=Images.Img_TintedBackgroundHighOpacity, DefaultDesiredSize := vector2{X:=0.0,Y:=0.0}}
    PetImage_AI_Slot3 : texture_block = texture_block{DefaultImage:=Images.Img_LineBackground, DefaultDesiredSize := vector2{X:=SizeImageX_PetHexaUI,Y:=SizeImageY_PetHexaUI}}
    Class_AI_Slot3 : text_block = text_block{DefaultTextColor:=White,DefaultShadowOffset:=ShadowOffset, DefaultShadowOpacity := 1.0}
    PackA_AI_Box3 : text_block = text_block{DefaultTextColor := White,DefaultShadowOffset:=ShadowOffset, DefaultShadowOpacity := 1.0}
    PackB_AI_Box3 : text_block = text_block{DefaultTextColor := White,DefaultShadowOffset:=ShadowOffset, DefaultShadowOpacity := 1.0}
    Purity_AI_Slot3 : text_block = text_block{DefaultTextColor:=White,DefaultShadowOffset:=ShadowOffset, DefaultShadowOpacity := 1.0}
    Health_AI_Slot3 : text_block = text_block{DefaultTextColor:=White,DefaultShadowOffset:=ShadowOffset, DefaultShadowOpacity := 1.0}

    #Slot 4 AI
    EmptyImage_AI_Slot4 : texture_block = texture_block{DefaultImage:=Images.EmptyTexture, DefaultDesiredSize := vector2{X:=SizeSlotX_PetHexaUI,Y:=SizeSlotY_PetHexaUI}}
    HealthLine_AI_Slot4 : texture_block = texture_block{DefaultTint:=Red, DefaultImage:=Images.Img_TintedBackgroundHighOpacity, DefaultDesiredSize := vector2{X:=0.0,Y:=0.0}}
    PetImage_AI_Slot4 : texture_block = texture_block{DefaultImage:=Images.Img_LineBackground, DefaultDesiredSize := vector2{X:=SizeImageX_PetHexaUI,Y:=SizeImageY_PetHexaUI}}
    Class_AI_Slot4 : text_block = text_block{DefaultTextColor:=White,DefaultShadowOffset:=ShadowOffset, DefaultShadowOpacity := 1.0}
    PackA_AI_Box4 : text_block = text_block{DefaultTextColor := White,DefaultShadowOffset:=ShadowOffset, DefaultShadowOpacity := 1.0}
    PackB_AI_Box4 : text_block = text_block{DefaultTextColor := White,DefaultShadowOffset:=ShadowOffset, DefaultShadowOpacity := 1.0}
    Purity_AI_Slot4 : text_block = text_block{DefaultTextColor:=White,DefaultShadowOffset:=ShadowOffset, DefaultShadowOpacity := 1.0}
    Health_AI_Slot4 : text_block = text_block{DefaultTextColor:=White,DefaultShadowOffset:=ShadowOffset, DefaultShadowOpacity := 1.0}

    #Slot 5 AI
    EmptyImage_AI_Slot5 : texture_block = texture_block{DefaultImage:=Images.EmptyTexture, DefaultDesiredSize := vector2{X:=SizeSlotX_PetHexaUI,Y:=SizeSlotY_PetHexaUI}}
    HealthLine_AI_Slot5 : texture_block = texture_block{DefaultTint:=Red, DefaultImage:=Images.Img_TintedBackgroundHighOpacity, DefaultDesiredSize := vector2{X:=0.0,Y:=0.0}}
    PetImage_AI_Slot5 : texture_block = texture_block{DefaultImage:=Images.Img_LineBackground, DefaultDesiredSize := vector2{X:=SizeImageX_PetHexaUI,Y:=SizeImageY_PetHexaUI}}
    Class_AI_Slot5 : text_block = text_block{DefaultTextColor:=White,DefaultShadowOffset:=ShadowOffset, DefaultShadowOpacity := 1.0}
    PackA_AI_Box5 : text_block = text_block{DefaultTextColor := White,DefaultShadowOffset:=ShadowOffset, DefaultShadowOpacity := 1.0}
    PackB_AI_Box5 : text_block = text_block{DefaultTextColor := White,DefaultShadowOffset:=ShadowOffset, DefaultShadowOpacity := 1.0}
    Purity_AI_Slot5 : text_block = text_block{DefaultTextColor:=White,DefaultShadowOffset:=ShadowOffset, DefaultShadowOpacity := 1.0}
    Health_AI_Slot5 : text_block = text_block{DefaultTextColor:=White,DefaultShadowOffset:=ShadowOffset, DefaultShadowOpacity := 1.0}

    var MaybeCanvas : ?canvas = false
    var UIIsDisplayed : logic = false

    var AllTextureBlock_HealthLines_Player : []texture_block = array{}
    var AllTextureBlock_PetImages_Player : []texture_block = array{}
    var AllTextBlock_Classes_Player : []text_block = array{}
    var AllTextBlock_PackA_Player : []text_block = array{}
    var AllTextBlock_PackB_Player : []text_block = array{}
    var AllTextBlock_Purity_Player : []text_block = array{}
    var AllTextBlock_Health_Player : []text_block = array{}
    
    var AllTextureBlock_HealthLines_AI : []texture_block = array{}
    var AllTextureBlock_PetImages_AI : []texture_block = array{}
    var AllTextBlock_Classes_AI : []text_block = array{}
    var AllTextBlock_PackA_AI : []text_block = array{}
    var AllTextBlock_PackB_AI : []text_block = array{}
    var AllTextBlock_Purity_AI : []text_block = array{}
    var AllTextBlock_Health_AI : []text_block = array{}

    var AllCombatReward_TextBlocks : []text_block = array{}
    var AllCombatReward_TextureBlocks : []texture_block = array{}

    #InfoSlider
    var VerticalSlider : float = 2125.0
    TextureSlider : texture_block = texture_block{DefaultImage:=Images.EmptyTexture, DefaultDesiredSize := vector2{Y:=1630.0}}
    CombatResult_TextBlock : text_block = text_block{DefaultTextColor:=Yellow,DefaultShadowOffset:=ShadowOffset, DefaultShadowOpacity := 1.0}
    CombatRewardDescript_TextBlock : text_block = text_block{DefaultTextColor:=White,DefaultShadowOffset:=ShadowOffset, DefaultShadowOpacity := 1.0}
    CombatReward1_TextBlock : text_block = text_block{DefaultTextColor:=White,DefaultShadowOffset:=ShadowOffset, DefaultShadowOpacity := 1.0}
    CombatReward1_TextureBlock : texture_block = texture_block{DefaultImage:=Images.Img_LineBackground, DefaultDesiredSize := vector2{X:=SizeSmallIconX,Y:=SizeSmallIconY}}
    CombatReward2_TextBlock : text_block = text_block{DefaultTextColor:=White,DefaultShadowOffset:=ShadowOffset, DefaultShadowOpacity := 1.0}
    CombatReward2_TextureBlock : texture_block = texture_block{DefaultImage:=Images.Img_LineBackground, DefaultDesiredSize := vector2{X:=SizeSmallIconX,Y:=SizeSmallIconY}}
    CombatReward3_TextBlock : text_block = text_block{DefaultTextColor:=White,DefaultShadowOffset:=ShadowOffset, DefaultShadowOpacity := 1.0}
    CombatReward3_TextureBlock : texture_block = texture_block{DefaultImage:=Images.Img_LineBackground, DefaultDesiredSize := vector2{X:=SizeSmallIconX,Y:=SizeSmallIconY}}
    CombatReward4_TextBlock : text_block = text_block{DefaultTextColor:=White,DefaultShadowOffset:=ShadowOffset, DefaultShadowOpacity := 1.0}
    CombatReward4_TextureBlock : texture_block = texture_block{DefaultImage:=Images.Img_LineBackground, DefaultDesiredSize := vector2{X:=SizeSmallIconX,Y:=SizeSmallIconY}}
    CombatReward5_TextBlock : text_block = text_block{DefaultTextColor:=White,DefaultShadowOffset:=ShadowOffset, DefaultShadowOpacity := 1.0}
    CombatReward5_TextureBlock : texture_block = texture_block{DefaultImage:=Images.Img_LineBackground, DefaultDesiredSize := vector2{X:=SizeSmallIconX,Y:=SizeSmallIconY}}
    CombatReward6_TextBlock : text_block = text_block{DefaultTextColor:=White,DefaultShadowOffset:=ShadowOffset, DefaultShadowOpacity := 1.0}
    CombatReward6_TextureBlock : texture_block = texture_block{DefaultImage:=Images.Img_LineBackground, DefaultDesiredSize := vector2{X:=SizeSmallIconX,Y:=SizeSmallIconY}}

    #Close
    CloseButton : button_loud = button_loud{}
    var ConfirmationToClose : int = 0

    InitUI() : void =
        CloseButton.SetText(BasicMessage("EXIT FIGHT "))

        set AllTextBlock_PackA_Player = array{PackA_Player_Box1,PackA_Player_Box2,PackA_Player_Box3,PackA_Player_Box4,PackA_Player_Box5}
        set AllTextBlock_PackB_Player = array{PackB_Player_Box1,PackB_Player_Box2,PackA_Player_Box3,PackB_Player_Box4,PackB_Player_Box5}

        set AllTextBlock_PackA_AI = array{PackA_AI_Box1,PackA_AI_Box2,PackA_AI_Box3,PackA_AI_Box4,PackA_AI_Box5}
        set AllTextBlock_PackB_AI = array{PackB_AI_Box1,PackB_AI_Box2,PackA_AI_Box3,PackB_AI_Box4,PackB_AI_Box5}

        set AllCombatReward_TextBlocks = array{CombatReward1_TextBlock,CombatReward2_TextBlock,CombatReward3_TextBlock,CombatReward4_TextBlock,CombatReward5_TextBlock,CombatReward6_TextBlock}
        set AllCombatReward_TextureBlocks = array{CombatReward1_TextureBlock,CombatReward2_TextureBlock,CombatReward3_TextureBlock,CombatReward4_TextureBlock,CombatReward5_TextureBlock,CombatReward6_TextureBlock}

        set AllTextureBlock_HealthLines_Player = array{HealthLine_Player_Slot1,HealthLine_Player_Slot2,HealthLine_Player_Slot3,HealthLine_Player_Slot4,HealthLine_Player_Slot5}
        set AllTextureBlock_HealthLines_AI = array{HealthLine_AI_Slot1,HealthLine_AI_Slot2,HealthLine_AI_Slot3,HealthLine_AI_Slot4,HealthLine_AI_Slot5}
        
        set AllTextureBlock_PetImages_Player = array{PetImage_Player_Slot1,PetImage_Player_Slot2,PetImage_Player_Slot3,PetImage_Player_Slot4,PetImage_Player_Slot5}
        set AllTextureBlock_PetImages_AI = array{PetImage_AI_Slot1,PetImage_AI_Slot2,PetImage_AI_Slot3,PetImage_AI_Slot4,PetImage_AI_Slot5}
        
        set AllTextBlock_Classes_Player = array{Class_Player_Slot1,Class_Player_Slot2,Class_Player_Slot3,Class_Player_Slot4,Class_Player_Slot5}
        set AllTextBlock_Classes_AI = array{Class_AI_Slot1,Class_AI_Slot2,Class_AI_Slot3,Class_AI_Slot4,Class_AI_Slot5}
        
        set AllTextBlock_Purity_Player = array{Purity_Player_Slot1,Purity_Player_Slot2,Purity_Player_Slot3,Purity_Player_Slot4,Purity_Player_Slot5}
        set AllTextBlock_Purity_AI = array{Purity_AI_Slot1,Purity_AI_Slot2,Purity_AI_Slot3,Purity_AI_Slot4,Purity_AI_Slot5}

        set AllTextBlock_Health_Player = array{Health_Player_Slot1,Health_Player_Slot2,Health_Player_Slot3,Health_Player_Slot4,Health_Player_Slot5}
        set AllTextBlock_Health_AI = array{Health_AI_Slot1,Health_AI_Slot2,Health_AI_Slot3,Health_AI_Slot4,Health_AI_Slot5}

        InitSubscribes()

        Canvas := MakeCanvas()
        set MaybeCanvas = option{Canvas}

    StartTime()<suspends> : void =
        set TotalGameTime = 59
        UpdateTimeRemaining()
        loop:
            Sleep(1.0)
            set TotalGameTime = TotalGameTime - 1
            UpdateTimeRemaining()
            if(TotalGameTime<=0):
                HexaMinigameManager.CheckGameEnd(true)
                break
            if(not HexaMinigameManager.GameIsRunning?):
                break

    InitSubscribes() : void =
        CloseButton.OnClick().Subscribe(TryToClose)

    TryToClose(Message : widget_message) : void =
        if(ConfirmationToClose=0):
            set ConfirmationToClose = ConfirmationToClose + 1
            CloseButton.SetText(BasicMessage("ARE YOU SURE?"))
        else:
            HexaMinigameManager.ResetLastGame(false) 
            spawn:
                DelayChangeUI_Selector_Fight(Message)
            

    DelayChangeUI_Selector_Fight(Message : widget_message)<suspends> : void =
        Remove_Canvas(Message)
        Sleep(0.1)
        if(Agent:=MaybeAgent?):
            HexaMinigameManager.Display_ChallengeCanvas()
        

    UpdateTimeRemaining() : void =
        RemainingFightTime.SetText(BasicMessage("{TotalGameTime}"))
        if(TotalGameTime<11):
            RemainingFightTime.SetTextColor(Red)
        else if(TotalGameTime<21):
            RemainingFightTime.SetTextColor(Yellow)
        else:
            RemainingFightTime.SetTextColor(Green)

    DisplayResult(AllAmounts : []float, AllTypes : []int, IsVictory : logic) : void =

        set ConfirmationToClose = ConfirmationToClose + 1
        CloseButton.SetText(BasicMessage("OK"))

        if(IsVictory?):
            CombatRewardDescript_TextBlock.SetText(BasicMessage("Rewards:"))
            CombatRewardDescript_TextBlock.SetTextColor(White)
            CombatResult_TextBlock.SetText(BasicMessage("VICTORY!"))
            CombatResult_TextBlock.SetTextColor(Gold)
        else:
            CombatRewardDescript_TextBlock.SetText(BasicMessage("Try again"))
            CombatRewardDescript_TextBlock.SetTextColor(Red)
            CombatResult_TextBlock.SetText(BasicMessage("DEFEAT"))

        for(Index:= 0..AllAmounts.Length-1, Amount := AllAmounts[Index], Type:= AllTypes[Index], TextBlock := AllCombatReward_TextBlocks[Index], TextureBlock := AllCombatReward_TextureBlocks[Index]):
            TextBlock.SetText(BasicMessage(GetOptimizedMagnitudeLetter(Amount)))
            if(Type=SAPPHIRES):
                TextBlock.SetTextColor(SapphiresTextColor)
                TextureBlock.SetImage(Images.Sapphire_Icon)
            else if(Type=WATER):
                TextBlock.SetTextColor(WaterTextColor)
                TextureBlock.SetImage(Images.Water_Icon)
            else if(Type=FLOUR):
                TextBlock.SetTextColor(FlourTextColor)
                TextureBlock.SetImage(Images.Flour_Icon)
            else if(Type=CHEESE):
                TextBlock.SetTextColor(CheeseTextColor)
                TextureBlock.SetImage(Images.Cheese_Icon)
            else if(Type=PIMENTOS):
                TextBlock.SetTextColor(PimentoTextColor)
                TextureBlock.SetImage(Images.Pimento_Icon)
            else if(Type=MUTANTMATERIAL):
                TextBlock.SetTextColor(MutantMaterialTextColor)
                TextureBlock.SetImage(Images.Mutant_Material_Icon)
            else if(Type=COAL):
                TextBlock.SetTextColor(CoalTextColor)
                TextureBlock.SetImage(Images.Coal_Icon)

        set VerticalSlider = 0.0
        TextureSlider.SetDesiredSize(vector2{Y:=VerticalSlider})
        RefreshAll()
        

    SetAgent(InAgent : agent) : void =
        Remove_CanvasWithoutMessage()
        set MaybeAgent = option{InAgent}

    RefreshAll() : void =

        AllAgentPets:=HexaMinigameManager.GetUsedPets()(0)
        AllAIPets:=HexaMinigameManager.GetUsedPets()(1)

        for(Index:=0..AllAgentPets.Length-1):
            if:
                TextureBlock_HealthLine_Player:=AllTextureBlock_HealthLines_Player[Index]
                TextureBlock_PetImages_Player:=AllTextureBlock_PetImages_Player[Index]
                TextBlock_PackageA_Player:=AllTextBlock_PackA_Player[Index]
                TextBlock_PackageB_Player:=AllTextBlock_PackB_Player[Index]
                TextBlock_Classes_Player:=AllTextBlock_Classes_Player[Index]
                TextureBlock_Purity_Player:=AllTextBlock_Purity_Player[Index]
                TextBlock_Health_Player:=AllTextBlock_Health_Player[Index]
            then:
                if(ValidPet := AllAgentPets[Index]):
                    SetPetClass(ValidPet, TextBlock_Classes_Player, TextBlock_PackageA_Player, TextBlock_PackageB_Player)
                    SetPetHealthBar(ValidPet,ValidPet.Health,TextureBlock_HealthLine_Player, TextBlock_Health_Player)
                    SetPetImage(ValidPet,TextureBlock_PetImages_Player)
                    SetPetPurity(ValidPet,ValidPet.Tier,TextureBlock_Purity_Player)
                else:
                    TextureBlock_HealthLine_Player.SetTint(White)

        for(Index:=0..AllAIPets.Length-1):
            if:
                TextureBlock_HealthLine_AI:=AllTextureBlock_HealthLines_AI[Index]
                TextureBlock_PetImages_AI:=AllTextureBlock_PetImages_AI[Index]
                TextBlock_PackageA_AI:=AllTextBlock_PackA_AI[Index]
                TextBlock_PackageB_AI:=AllTextBlock_PackB_AI[Index]
                TextBlock_Classes_AI:=AllTextBlock_Classes_AI[Index]
                TextureBlock_Purity_AI:=AllTextBlock_Purity_AI[Index]
                TextBlock_Health_AI:=AllTextBlock_Health_AI[Index]
            then:
                if(ValidPet := AllAIPets[Index]):
                    SetPetClass(ValidPet, TextBlock_Classes_AI, TextBlock_PackageA_AI, TextBlock_PackageB_AI)
                    TextureBlock_HealthLine_AI.SetTint(Red)
                    SetPetHealthBar(ValidPet,ValidPet.Health,TextureBlock_HealthLine_AI, TextBlock_Health_AI)
                    SetPetImage(ValidPet,TextureBlock_PetImages_AI)
                    SetPetPurity(ValidPet,ValidPet.Tier,TextureBlock_Purity_AI)
                else:
                    TextureBlock_HealthLine_AI.SetTint(White)

    CleanUI() : void =
        for(Index:=0..AllTextureBlock_HealthLines_Player.Length-1):
            if:
                TextureBlock_HealthLine_Player:=AllTextureBlock_HealthLines_Player[Index]
                TextureBlock_PetImages_Player:=AllTextureBlock_PetImages_Player[Index]
                TextBlock_Classes_Player:=AllTextBlock_Classes_Player[Index]
                TextBlock_PackageA_Player:=AllTextBlock_PackA_Player[Index]
                TextBlock_PackageB_Player:=AllTextBlock_PackB_Player[Index]
                TextureBlock_Purity_Player:=AllTextBlock_Purity_Player[Index]
                TextBlock_Health_Player:=AllTextBlock_Health_Player[Index]
            then:
                TextBlock_PackageA_Player.SetText(BasicMessage(""))
                TextBlock_PackageB_Player.SetText(BasicMessage(""))
                TextBlock_Classes_Player.SetText(BasicMessage(""))
                TextureBlock_HealthLine_Player.SetImage(Images.EmptyTexture)
                TextBlock_Health_Player.SetText(BasicMessage(""))
                TextureBlock_PetImages_Player.SetImage(Images.EmptyTexture)
                TextureBlock_Purity_Player.SetText(BasicMessage(""))
               

        for(Index:=0..AllTextureBlock_HealthLines_AI.Length-1):
            if:
                TextureBlock_HealthLine_AI:=AllTextureBlock_HealthLines_AI[Index]
                TextureBlock_PetImages_AI:=AllTextureBlock_PetImages_AI[Index]
                TextBlock_Classes_AI:=AllTextBlock_Classes_AI[Index]
                TextBlock_PackageA_AI:=AllTextBlock_PackA_AI[Index]
                TextBlock_PackageB_AI:=AllTextBlock_PackB_AI[Index]
                TextureBlock_Purity_AI:=AllTextBlock_Purity_AI[Index]
                TextBlock_Health_AI:=AllTextBlock_Health_AI[Index]
            then:
                TextBlock_PackageA_AI.SetText(BasicMessage(""))
                TextBlock_PackageB_AI.SetText(BasicMessage(""))
                TextBlock_Classes_AI.SetText(BasicMessage(""))
                TextureBlock_HealthLine_AI.SetImage(Images.EmptyTexture)
                TextBlock_Health_AI.SetText(BasicMessage(""))
                TextureBlock_PetImages_AI.SetImage(Images.EmptyTexture)
                TextureBlock_Purity_AI.SetText(BasicMessage(""))


    Remove_Canvas(Message : widget_message) : void =
        set UIIsDisplayed = false
        ##Print("Trying to remove UI")
        if:
            Agent:=MaybeAgent?
            Player:=player[Agent]
            PlayerUI:=GetPlayerUI[Player]
            Canvas := MaybeCanvas?
        then:
            PlayerUI.RemoveWidget(Canvas)
        RemoveDelay_OnAdd()
     

    Remove_CanvasWithoutMessage() : void =
        set UIIsDisplayed = false
        if:
            Agent:=MaybeAgent?
            Player:=player[Agent]
            PlayerUI:=GetPlayerUI[Player]
            Canvas := MaybeCanvas?
        then:
            PlayerUI.RemoveWidget(Canvas)
        RemoveDelay_OnRemove()

    RemoveDelay_OnAdd() : void =
        CloseButton.SetText(BasicMessage("EXIT FIGHT"))

    RemoveDelay_OnRemove() : void =
        CloseButton.SetText(BasicMessage("EXIT FIGHT "))
  
    Display_Canvas() : void =
        if(not UIIsDisplayed?):
            set ConfirmationToClose = 0
            set UIIsDisplayed = true
            set VerticalSlider = 2500.0#1800.0
            TextureSlider.SetDesiredSize(vector2{Y:=VerticalSlider})
            CloseButton.SetText(BasicMessage("EXIT FIGHT"))
            var Index : int = 0 
            if:
                Agent:=MaybeAgent?
                PlayerUI:=GetPlayerUI[player[Agent]]
                Canvas := MaybeCanvas?
            then:
                #Print("Displaying canvas")
                PlayerUI.AddWidget(Canvas,player_ui_slot{ZOrder:=0, InputMode := ui_input_mode.All})
   

    MakeCanvas(): canvas=
        Canvas := canvas:
            Slots:=array:
                canvas_slot:
                    Anchors:=anchors{Minimum:=vector2{X:=0.5,Y:=1.0},Maximum:=vector2{X:=0.5,Y:=1.0}}
                    Offsets:=margin{Left:=0.0,Top:=-30.0,Right:=0.0,Bottom:=0.0} #PositionX, PositionY, SizeX, SizeY
                    Alignment:=vector2{X:=0.5,Y:=1.0}
                    ZOrder:=0
                    SizeToContent:=true   
                    Widget:=overlay:
                        Slots:=array:
                            overlay_slot:
                                HorizontalAlignment:=horizontal_alignment.Center
                                VerticalAlignment:=vertical_alignment.Center
                                Widget:=texture_block{DefaultImage:=Images.EmptyTexture, DefaultDesiredSize := vector2{X:=225.0,Y:=125.0}}
                            overlay_slot:
                                HorizontalAlignment:=horizontal_alignment.Fill
                                VerticalAlignment:=vertical_alignment.Fill
                                Widget:=CloseButton
                #Main overlay
                canvas_slot:
                    Anchors:=anchors{Minimum:=vector2{X:=0.5,Y:=0.0},Maximum:=vector2{X:=0.5,Y:=0.0}}
                    Offsets:=margin{Left:=0.0,Top:=10.0,Right:=100.0,Bottom:=30.0} #PositionX, PositionY, SizeX, SizeY
                    Alignment:=vector2{X:=0.5,Y:=0.0}
                    ZOrder:=0
                    SizeToContent:=true
                    Widget:=overlay:
                        Slots:=array:
                            overlay_slot:
                                HorizontalAlignment:=horizontal_alignment.Fill
                                VerticalAlignment:=vertical_alignment.Fill
                                Widget:=texture_block{DefaultTint :=Black,DefaultImage:=Images.Img_TintedBackgroundHighOpacity}
                            overlay_slot:
                                HorizontalAlignment:=horizontal_alignment.Fill
                                VerticalAlignment:=vertical_alignment.Fill
                                Padding:=margin{Left:=SelectorBoxPadding,Right:=SelectorBoxPadding,Top:=SelectorBoxPadding,Bottom:=SelectorBoxPadding}
                                Widget:=texture_block{DefaultTint:=Black, DefaultImage:=Images.Img_TintedBackgroundHighOpacity}
                            overlay_slot: #Screw 0
                                Padding:=margin{Left:=3.5,Top:=3.5,Right:=3.5,Bottom:=3.5}
                                HorizontalAlignment:=horizontal_alignment.Left
                                VerticalAlignment:=vertical_alignment.Top
                                Widget:=texture_block{DefaultImage:=Images.Dialogue_Bolt, DefaultDesiredSize:=vector2{X:=5.0,Y:=5.0}}
                            overlay_slot: #Screw 1
                                Padding:=margin{Left:=3.5,Top:=3.5,Right:=3.5,Bottom:=3.5}
                                HorizontalAlignment:=horizontal_alignment.Right
                                VerticalAlignment:=vertical_alignment.Top
                                Widget:=texture_block{DefaultImage:=Images.Dialogue_Bolt, DefaultDesiredSize:=vector2{X:=5.0,Y:=5.0}}
                            overlay_slot: #Screw 2
                                Padding:=margin{Left:=3.5,Top:=3.5,Right:=3.5,Bottom:=3.5}
                                HorizontalAlignment:=horizontal_alignment.Left
                                VerticalAlignment:=vertical_alignment.Bottom
                                Widget:=texture_block{DefaultImage:=Images.Dialogue_Bolt, DefaultDesiredSize:=vector2{X:=5.0,Y:=5.0}}
                            overlay_slot: #Screw 3
                                Padding:=margin{Left:=3.5,Top:=3.5,Right:=3.5,Bottom:=3.5}
                                HorizontalAlignment:=horizontal_alignment.Right
                                VerticalAlignment:=vertical_alignment.Bottom
                                Widget:=texture_block{DefaultImage:=Images.Dialogue_Bolt, DefaultDesiredSize:=vector2{X:=5.0,Y:=5.0}}
                            overlay_slot:
                                HorizontalAlignment:=horizontal_alignment.Center
                                VerticalAlignment:=vertical_alignment.Center
                                Widget:=texture_block{DefaultImage:=Images.EmptyTexture, DefaultDesiredSize := vector2{X:=225.0,Y:=100.0}}
                            overlay_slot:
                                HorizontalAlignment:=horizontal_alignment.Fill
                                VerticalAlignment:=vertical_alignment.Fill
                                Padding:=margin{Left:=5.0,Right:=5.0,Top:=5.0,Bottom:=5.0}
                                Widget:=stack_box:  
                                    Orientation:=orientation.Vertical
                                    Slots:=array:
                                        stack_box_slot: 
                                            Widget:=text_block{DefaultText:= BasicMessage("Remaining time"),DefaultTextColor:=White, DefaultShadowOffset:=ShadowOffset, DefaultShadowOpacity := DefaultTextShadowOpacity}
                                        stack_box_slot: 
                                            Widget:=RemainingFightTime   
                canvas_slot:
                    Anchors:=anchors{Minimum:=vector2{X:=0.5,Y:=0.5},Maximum:=vector2{X:=0.5,Y:=0.5}}
                    ZOrder:=0
                    SizeToContent:=true
                    Offsets:=margin{Left:=0.0,Top:=0.0,Right:=0.0,Bottom:=0.0} #PositionX, PositionY, SizeX, SizeY
                    Alignment:=vector2{X:=0.5,Y:=0.5}
                    Widget:=overlay:
                        Slots:=array:
                            overlay_slot:
                                HorizontalAlignment:=horizontal_alignment.Center
                                VerticalAlignment:=vertical_alignment.Center
                                Widget:=TextureSlider
                            overlay_slot:
                                HorizontalAlignment:=horizontal_alignment.Center
                                VerticalAlignment:=vertical_alignment.Top
                                Widget:= overlay:
                                    Slots:=array:
                                        overlay_slot:
                                            HorizontalAlignment:=horizontal_alignment.Fill
                                            VerticalAlignment:=vertical_alignment.Fill
                                            Widget:=texture_block{DefaultTint:=Black,DefaultImage:=Images.Img_TintedBackgroundHighOpacity}
                                        overlay_slot:
                                            HorizontalAlignment:=horizontal_alignment.Fill
                                            VerticalAlignment:=vertical_alignment.Fill
                                            Padding:=margin{Left:=SelectorBoxPadding,Right:=SelectorBoxPadding,Top:=SelectorBoxPadding,Bottom:=SelectorBoxPadding}
                                            Widget:=texture_block{DefaultTint:=Black, DefaultImage:=Images.Img_TintedBackgroundHighOpacity}
                                        overlay_slot: #Screw 0
                                            Padding:=margin{Left:=3.5,Top:=3.5,Right:=3.5,Bottom:=3.5}
                                            HorizontalAlignment:=horizontal_alignment.Left
                                            VerticalAlignment:=vertical_alignment.Top
                                            Widget:=texture_block{DefaultImage:=Images.Dialogue_Bolt, DefaultDesiredSize:=vector2{X:=5.0,Y:=5.0}}
                                        overlay_slot: #Screw 1
                                            Padding:=margin{Left:=3.5,Top:=3.5,Right:=3.5,Bottom:=3.5}
                                            HorizontalAlignment:=horizontal_alignment.Right
                                            VerticalAlignment:=vertical_alignment.Top
                                            Widget:=texture_block{DefaultImage:=Images.Dialogue_Bolt, DefaultDesiredSize:=vector2{X:=5.0,Y:=5.0}}
                                        overlay_slot: #Screw 2
                                            Padding:=margin{Left:=3.5,Top:=3.5,Right:=3.5,Bottom:=3.5}
                                            HorizontalAlignment:=horizontal_alignment.Left
                                            VerticalAlignment:=vertical_alignment.Bottom
                                            Widget:=texture_block{DefaultImage:=Images.Dialogue_Bolt, DefaultDesiredSize:=vector2{X:=5.0,Y:=5.0}}
                                        overlay_slot: #Screw 3
                                            Padding:=margin{Left:=3.5,Top:=3.5,Right:=3.5,Bottom:=3.5}
                                            HorizontalAlignment:=horizontal_alignment.Right
                                            VerticalAlignment:=vertical_alignment.Bottom
                                            Widget:=texture_block{DefaultImage:=Images.Dialogue_Bolt, DefaultDesiredSize:=vector2{X:=5.0,Y:=5.0}}
                                        overlay_slot:
                                            HorizontalAlignment:=horizontal_alignment.Center
                                            VerticalAlignment:=vertical_alignment.Center
                                            Padding:=margin{Left:=5.0,Top:=5.0,Right:=5.0,Bottom:=5.0} #PositionX, PositionY, SizeX, SizeY
                                            Widget:= stack_box:
                                                Orientation:=orientation.Vertical
                                                    Slots:=array:
                                                        stack_box_slot:
                                                            Padding:=margin{Right:=2.5,Left:=2.5,Top:=2.5,Bottom:=2.5}
                                                            HorizontalAlignment:=horizontal_alignment.Center
                                                            VerticalAlignment:=vertical_alignment.Center
                                                            Widget:=CombatResult_TextBlock
                                                        stack_box_slot:
                                                            Padding:=margin{Right:=2.5,Left:=2.5,Top:=2.5,Bottom:=2.5}
                                                            HorizontalAlignment:=horizontal_alignment.Center
                                                            VerticalAlignment:=vertical_alignment.Center
                                                            Widget:=CombatRewardDescript_TextBlock
                                                        stack_box_slot:
                                                            Padding:=margin{Right:=2.5,Left:=2.5,Top:=2.5,Bottom:=2.5}
                                                            HorizontalAlignment:=horizontal_alignment.Left
                                                            VerticalAlignment:=vertical_alignment.Center
                                                            Widget:= stack_box:
                                                                Orientation:=orientation.Horizontal
                                                                    Slots:=array:
                                                                        stack_box_slot:
                                                                            Padding:=margin{Right:=5.0,Left:=5.0,Top:=5.0,Bottom:=5.0}
                                                                            HorizontalAlignment:=horizontal_alignment.Left
                                                                            VerticalAlignment:=vertical_alignment.Center
                                                                            Widget:=CombatReward1_TextureBlock
                                                                        stack_box_slot:
                                                                            Padding:=margin{Right:=5.0,Left:=5.0,Top:=5.0,Bottom:=5.0}
                                                                            HorizontalAlignment:=horizontal_alignment.Left
                                                                            VerticalAlignment:=vertical_alignment.Center
                                                                            Widget:=CombatReward1_TextBlock
                                                        stack_box_slot:
                                                            Padding:=margin{Right:=2.5,Left:=2.5,Top:=2.5,Bottom:=2.5}
                                                            HorizontalAlignment:=horizontal_alignment.Left
                                                            VerticalAlignment:=vertical_alignment.Center
                                                            Widget:= stack_box:
                                                                Orientation:=orientation.Horizontal
                                                                    Slots:=array:
                                                                        stack_box_slot:
                                                                            Padding:=margin{Right:=5.0,Left:=5.0,Top:=5.0,Bottom:=5.0}
                                                                            HorizontalAlignment:=horizontal_alignment.Left
                                                                            VerticalAlignment:=vertical_alignment.Center
                                                                            Widget:=CombatReward2_TextureBlock
                                                                        stack_box_slot:
                                                                            Padding:=margin{Right:=5.0,Left:=5.0,Top:=5.0,Bottom:=5.0}
                                                                            HorizontalAlignment:=horizontal_alignment.Left
                                                                            VerticalAlignment:=vertical_alignment.Center
                                                                            Widget:=CombatReward2_TextBlock
                                                        stack_box_slot:
                                                            Padding:=margin{Right:=2.5,Left:=2.5,Top:=2.5,Bottom:=2.5}
                                                            HorizontalAlignment:=horizontal_alignment.Left
                                                            VerticalAlignment:=vertical_alignment.Center
                                                            Widget:= stack_box:
                                                                Orientation:=orientation.Horizontal
                                                                    Slots:=array:
                                                                        stack_box_slot:
                                                                            Padding:=margin{Right:=5.0,Left:=5.0,Top:=5.0,Bottom:=5.0}
                                                                            HorizontalAlignment:=horizontal_alignment.Left
                                                                            VerticalAlignment:=vertical_alignment.Center
                                                                            Widget:=CombatReward3_TextureBlock
                                                                        stack_box_slot:
                                                                            Padding:=margin{Right:=5.0,Left:=5.0,Top:=5.0,Bottom:=5.0}
                                                                            HorizontalAlignment:=horizontal_alignment.Left
                                                                            VerticalAlignment:=vertical_alignment.Center
                                                                            Widget:=CombatReward3_TextBlock    
                                                                            
                                                        stack_box_slot:
                                                            Padding:=margin{Right:=2.5,Left:=2.5,Top:=2.5,Bottom:=2.5}
                                                            HorizontalAlignment:=horizontal_alignment.Left
                                                            VerticalAlignment:=vertical_alignment.Center
                                                            Widget:= stack_box:
                                                                Orientation:=orientation.Horizontal
                                                                    Slots:=array:
                                                                        stack_box_slot:
                                                                            Padding:=margin{Right:=5.0,Left:=5.0,Top:=5.0,Bottom:=5.0}
                                                                            HorizontalAlignment:=horizontal_alignment.Left
                                                                            VerticalAlignment:=vertical_alignment.Center
                                                                            Widget:=CombatReward4_TextureBlock
                                                                        stack_box_slot:
                                                                            Padding:=margin{Right:=5.0,Left:=5.0,Top:=5.0,Bottom:=5.0}
                                                                            HorizontalAlignment:=horizontal_alignment.Left
                                                                            VerticalAlignment:=vertical_alignment.Center
                                                                            Widget:=CombatReward4_TextBlock   

                                                        stack_box_slot:
                                                            Padding:=margin{Right:=2.5,Left:=2.5,Top:=2.5,Bottom:=2.5}
                                                            HorizontalAlignment:=horizontal_alignment.Left
                                                            VerticalAlignment:=vertical_alignment.Center
                                                            Widget:= stack_box:
                                                                Orientation:=orientation.Horizontal
                                                                    Slots:=array:
                                                                        stack_box_slot:
                                                                            Padding:=margin{Right:=5.0,Left:=5.0,Top:=5.0,Bottom:=5.0}
                                                                            HorizontalAlignment:=horizontal_alignment.Left
                                                                            VerticalAlignment:=vertical_alignment.Center
                                                                            Widget:=CombatReward5_TextureBlock
                                                                        stack_box_slot:
                                                                            Padding:=margin{Right:=5.0,Left:=5.0,Top:=5.0,Bottom:=5.0}
                                                                            HorizontalAlignment:=horizontal_alignment.Left
                                                                            VerticalAlignment:=vertical_alignment.Center
                                                                            Widget:=CombatReward5_TextBlock   

                                                        stack_box_slot:
                                                            Padding:=margin{Right:=2.5,Left:=2.5,Top:=2.5,Bottom:=2.5}
                                                            HorizontalAlignment:=horizontal_alignment.Left
                                                            VerticalAlignment:=vertical_alignment.Center
                                                            Widget:= stack_box:
                                                                Orientation:=orientation.Horizontal
                                                                    Slots:=array:
                                                                        stack_box_slot:
                                                                            Padding:=margin{Right:=5.0,Left:=5.0,Top:=5.0,Bottom:=5.0}
                                                                            HorizontalAlignment:=horizontal_alignment.Left
                                                                            VerticalAlignment:=vertical_alignment.Center
                                                                            Widget:=CombatReward6_TextureBlock
                                                                        stack_box_slot:
                                                                            Padding:=margin{Right:=5.0,Left:=5.0,Top:=5.0,Bottom:=5.0}
                                                                            HorizontalAlignment:=horizontal_alignment.Left
                                                                            VerticalAlignment:=vertical_alignment.Center
                                                                            Widget:=CombatReward6_TextBlock 
                canvas_slot:
                    Anchors:=anchors{Minimum:=vector2{X:=0.0,Y:=0.5},Maximum:=vector2{X:=0.0,Y:=0.5}}
                    ZOrder:=0
                    SizeToContent:=true
                    Offsets:=margin{Left:=5.0,Top:=120.0,Right:=0.0,Bottom:=0.0} #PositionX, PositionY, SizeX, SizeY
                    Alignment:=vector2{X:=0.0,Y:=0.5}
                    Widget:=overlay:
                        Slots:=array:
                            overlay_slot:
                                HorizontalAlignment:=horizontal_alignment.Fill
                                VerticalAlignment:=vertical_alignment.Fill
                                Widget:=stack_box:
                                    Orientation:=orientation.Vertical
                                        Slots:=array:
                                            stack_box_slot: #Slot 1
                                                Padding:=margin{Top:=ResourcersRowTopPadding,Bottom:=ResourcersRowBottomPadding}
                                                HorizontalAlignment:=horizontal_alignment.Fill
                                                VerticalAlignment:=vertical_alignment.Fill
                                                Widget:=overlay:
                                                    Slots:=array:
                                                        overlay_slot:
                                                            HorizontalAlignment:=horizontal_alignment.Fill
                                                            VerticalAlignment:=vertical_alignment.Fill
                                                            Widget:=EmptyImage_Player_Slot1
                                                        overlay_slot:
                                                            HorizontalAlignment:=horizontal_alignment.Fill
                                                            VerticalAlignment:=vertical_alignment.Fill
                                                            Widget:= texture_block{DefaultImage:=Images.Img_LineBackground}
                                                        overlay_slot:
                                                            HorizontalAlignment:=horizontal_alignment.Fill
                                                            VerticalAlignment:=vertical_alignment.Fill
                                                            Padding:=margin{Left:=PetsLineBoxPadding,Top:=PetsLineBoxPadding,Right:=PetsLineBoxPadding,Bottom:=PetsLineBoxPadding}
                                                            Widget:= texture_block{DefaultImage:=Images.Img_LineBackground}
                                                        overlay_slot:
                                                            HorizontalAlignment:=horizontal_alignment.Left
                                                            VerticalAlignment:=vertical_alignment.Fill
                                                            Padding:=margin{Left:=PetsLineBoxPadding,Top:=PetsLineBoxPadding,Right:=PetsLineBoxPadding,Bottom:=PetsLineBoxPadding}
                                                            Widget:=HealthLine_Player_Slot1
                                                        overlay_slot:
                                                            HorizontalAlignment:=horizontal_alignment.Left
                                                            VerticalAlignment:=vertical_alignment.Center
                                                            Widget:=overlay:
                                                                Slots:=array:
                                                                    overlay_slot:
                                                                        HorizontalAlignment:=horizontal_alignment.Center
                                                                        VerticalAlignment:=vertical_alignment.Center
                                                                        Padding:=margin{Left:=7.0,Top:=PetsLineBoxPadding,Right:=PetsLineBoxPadding,Bottom:=PetsLineBoxPadding}
                                                                        Widget:=PetImage_Player_Slot1
                                                        overlay_slot:
                                                            HorizontalAlignment:=horizontal_alignment.Left
                                                            VerticalAlignment:=vertical_alignment.Center
                                                            Padding:=margin{Left:=TextPetsBoxPadding}
                                                            Widget:=stack_box:
                                                                Orientation:=orientation.Vertical
                                                                    Slots:=array:
                                                                        stack_box_slot: #Class
                                                                            HorizontalAlignment:=horizontal_alignment.Left
                                                                            VerticalAlignment:=vertical_alignment.Center
                                                                            Padding:=margin{Left:=TextPetsBoxPadding}
                                                                            Widget:=stack_box:
                                                                                Orientation:=orientation.Horizontal
                                                                                Slots:=array:
                                                                                    stack_box_slot: 
                                                                                        Padding:=margin{Left:=2.5,Right:=2.5}
                                                                                        Widget:=Class_Player_Slot1
                                                                                    stack_box_slot: 
                                                                                        Padding:=margin{Left:=2.5, Right:=1.0}
                                                                                        Widget:=PackA_Player_Box1
                                                                                    stack_box_slot: 
                                                                                        Padding:=margin{Right:=2.5}
                                                                                        Widget:=PackB_Player_Box1
                                                                        stack_box_slot: #Purity
                                                                            HorizontalAlignment:=horizontal_alignment.Left
                                                                            VerticalAlignment:=vertical_alignment.Center
                                                                            Padding:=margin{Left:=TextPetsBoxPadding}
                                                                            Widget:=Purity_Player_Slot1
                                                                        stack_box_slot: #Health
                                                                            HorizontalAlignment:=horizontal_alignment.Left
                                                                            VerticalAlignment:=vertical_alignment.Center
                                                                            Padding:=margin{Left:=TextPetsBoxPadding}
                                                                            Widget:=Health_Player_Slot1
                                            stack_box_slot: #Slot 2
                                                Padding:=margin{Top:=ResourcersRowTopPadding,Bottom:=ResourcersRowBottomPadding}
                                                HorizontalAlignment:=horizontal_alignment.Fill
                                                VerticalAlignment:=vertical_alignment.Fill
                                                Widget:=overlay:
                                                    Slots:=array:
                                                        overlay_slot:
                                                            HorizontalAlignment:=horizontal_alignment.Fill
                                                            VerticalAlignment:=vertical_alignment.Fill
                                                            Widget:=EmptyImage_Player_Slot2
                                                        overlay_slot:
                                                            HorizontalAlignment:=horizontal_alignment.Fill
                                                            VerticalAlignment:=vertical_alignment.Fill
                                                            Widget:= texture_block{DefaultImage:=Images.Img_LineBackground}
                                                        overlay_slot:
                                                            HorizontalAlignment:=horizontal_alignment.Fill
                                                            VerticalAlignment:=vertical_alignment.Fill
                                                            Padding:=margin{Left:=PetsLineBoxPadding,Top:=PetsLineBoxPadding,Right:=PetsLineBoxPadding,Bottom:=PetsLineBoxPadding}
                                                            Widget:= texture_block{DefaultImage:=Images.Img_LineBackground}
                                                        overlay_slot:
                                                            HorizontalAlignment:=horizontal_alignment.Left
                                                            VerticalAlignment:=vertical_alignment.Fill
                                                            Padding:=margin{Left:=PetsLineBoxPadding,Top:=PetsLineBoxPadding,Right:=PetsLineBoxPadding,Bottom:=PetsLineBoxPadding}
                                                            Widget:=HealthLine_Player_Slot2
                                                        overlay_slot:
                                                            HorizontalAlignment:=horizontal_alignment.Left
                                                            VerticalAlignment:=vertical_alignment.Center
                                                            Widget:=overlay:
                                                                Slots:=array:
                                                                    overlay_slot:
                                                                        HorizontalAlignment:=horizontal_alignment.Center
                                                                        VerticalAlignment:=vertical_alignment.Center
                                                                        Padding:=margin{Left:=7.0,Top:=PetsLineBoxPadding,Right:=PetsLineBoxPadding,Bottom:=PetsLineBoxPadding}
                                                                        Widget:=PetImage_Player_Slot2
                                                        overlay_slot:
                                                            HorizontalAlignment:=horizontal_alignment.Left
                                                            VerticalAlignment:=vertical_alignment.Center
                                                            Padding:=margin{Left:=TextPetsBoxPadding}
                                                            Widget:=stack_box:
                                                                Orientation:=orientation.Vertical
                                                                    Slots:=array:
                                                                        stack_box_slot: #Class
                                                                            HorizontalAlignment:=horizontal_alignment.Left
                                                                            VerticalAlignment:=vertical_alignment.Center
                                                                            Padding:=margin{Left:=TextPetsBoxPadding}
                                                                            Widget:=stack_box:
                                                                                Orientation:=orientation.Horizontal
                                                                                Slots:=array:
                                                                                    stack_box_slot: 
                                                                                        Padding:=margin{Left:=2.5,Right:=2.5}
                                                                                        Widget:=Class_Player_Slot2
                                                                                    stack_box_slot: 
                                                                                        Padding:=margin{Left:=2.5, Right:=1.0}
                                                                                        Widget:=PackA_Player_Box2
                                                                                    stack_box_slot: 
                                                                                        Padding:=margin{Right:=2.5}
                                                                                        Widget:=PackB_Player_Box2
                                                                        stack_box_slot: #Purity
                                                                            HorizontalAlignment:=horizontal_alignment.Left
                                                                            VerticalAlignment:=vertical_alignment.Center
                                                                            Padding:=margin{Left:=TextPetsBoxPadding}
                                                                            Widget:=Purity_Player_Slot2
                                                                        stack_box_slot: #Health
                                                                            HorizontalAlignment:=horizontal_alignment.Left
                                                                            VerticalAlignment:=vertical_alignment.Center
                                                                            Padding:=margin{Left:=TextPetsBoxPadding}
                                                                            Widget:=Health_Player_Slot2
                                            stack_box_slot: #Slot 3
                                                Padding:=margin{Top:=ResourcersRowTopPadding,Bottom:=ResourcersRowBottomPadding}
                                                HorizontalAlignment:=horizontal_alignment.Fill
                                                VerticalAlignment:=vertical_alignment.Fill
                                                Widget:=overlay:
                                                    Slots:=array:
                                                        overlay_slot:
                                                            HorizontalAlignment:=horizontal_alignment.Fill
                                                            VerticalAlignment:=vertical_alignment.Fill
                                                            Widget:=EmptyImage_Player_Slot3
                                                        overlay_slot:
                                                            HorizontalAlignment:=horizontal_alignment.Fill
                                                            VerticalAlignment:=vertical_alignment.Fill
                                                            Widget:= texture_block{DefaultImage:=Images.Img_LineBackground}
                                                        overlay_slot:
                                                            HorizontalAlignment:=horizontal_alignment.Fill
                                                            VerticalAlignment:=vertical_alignment.Fill
                                                            Padding:=margin{Left:=PetsLineBoxPadding,Top:=PetsLineBoxPadding,Right:=PetsLineBoxPadding,Bottom:=PetsLineBoxPadding}
                                                            Widget:= texture_block{DefaultImage:=Images.Img_LineBackground}
                                                        overlay_slot:
                                                            HorizontalAlignment:=horizontal_alignment.Left
                                                            VerticalAlignment:=vertical_alignment.Fill
                                                            Padding:=margin{Left:=PetsLineBoxPadding,Top:=PetsLineBoxPadding,Right:=PetsLineBoxPadding,Bottom:=PetsLineBoxPadding}
                                                            Widget:=HealthLine_Player_Slot3
                                                        overlay_slot:
                                                            HorizontalAlignment:=horizontal_alignment.Left
                                                            VerticalAlignment:=vertical_alignment.Center
                                                            Widget:=overlay:
                                                                Slots:=array:
                                                                    overlay_slot:
                                                                        HorizontalAlignment:=horizontal_alignment.Center
                                                                        VerticalAlignment:=vertical_alignment.Center
                                                                        Padding:=margin{Left:=7.0,Top:=PetsLineBoxPadding,Right:=PetsLineBoxPadding,Bottom:=PetsLineBoxPadding}
                                                                        Widget:=PetImage_Player_Slot3
                                                        overlay_slot:
                                                            HorizontalAlignment:=horizontal_alignment.Left
                                                            VerticalAlignment:=vertical_alignment.Center
                                                            Padding:=margin{Left:=TextPetsBoxPadding}
                                                            Widget:=stack_box:
                                                                Orientation:=orientation.Vertical
                                                                    Slots:=array:
                                                                        stack_box_slot: #Class
                                                                            HorizontalAlignment:=horizontal_alignment.Left
                                                                            VerticalAlignment:=vertical_alignment.Center
                                                                            Padding:=margin{Left:=TextPetsBoxPadding}
                                                                            Widget:=stack_box:
                                                                                Orientation:=orientation.Horizontal
                                                                                Slots:=array:
                                                                                    stack_box_slot: 
                                                                                        Padding:=margin{Left:=2.5,Right:=2.5}
                                                                                        Widget:=Class_Player_Slot3
                                                                                    stack_box_slot: 
                                                                                        Padding:=margin{Left:=2.5, Right:=1.0}
                                                                                        Widget:=PackA_Player_Box3
                                                                                    stack_box_slot: 
                                                                                        Padding:=margin{Right:=2.5}
                                                                                        Widget:=PackB_Player_Box3
                                                                        stack_box_slot: #Purity
                                                                            HorizontalAlignment:=horizontal_alignment.Left
                                                                            VerticalAlignment:=vertical_alignment.Center
                                                                            Padding:=margin{Left:=TextPetsBoxPadding}
                                                                            Widget:=Purity_Player_Slot3
                                                                        stack_box_slot: #Health
                                                                            HorizontalAlignment:=horizontal_alignment.Left
                                                                            VerticalAlignment:=vertical_alignment.Center
                                                                            Padding:=margin{Left:=TextPetsBoxPadding}
                                                                            Widget:=Health_Player_Slot3
                                            stack_box_slot: #Slot 4
                                                Padding:=margin{Top:=ResourcersRowTopPadding,Bottom:=ResourcersRowBottomPadding}
                                                HorizontalAlignment:=horizontal_alignment.Fill
                                                VerticalAlignment:=vertical_alignment.Fill
                                                Widget:=overlay:
                                                    Slots:=array:
                                                        overlay_slot:
                                                            HorizontalAlignment:=horizontal_alignment.Fill
                                                            VerticalAlignment:=vertical_alignment.Fill
                                                            Widget:=EmptyImage_Player_Slot4
                                                        overlay_slot:
                                                            HorizontalAlignment:=horizontal_alignment.Fill
                                                            VerticalAlignment:=vertical_alignment.Fill
                                                            Widget:= texture_block{DefaultImage:=Images.Img_LineBackground}
                                                        overlay_slot:
                                                            HorizontalAlignment:=horizontal_alignment.Fill
                                                            VerticalAlignment:=vertical_alignment.Fill
                                                            Padding:=margin{Left:=PetsLineBoxPadding,Top:=PetsLineBoxPadding,Right:=PetsLineBoxPadding,Bottom:=PetsLineBoxPadding}
                                                            Widget:= texture_block{DefaultImage:=Images.Img_LineBackground}
                                                        overlay_slot:
                                                            HorizontalAlignment:=horizontal_alignment.Left
                                                            VerticalAlignment:=vertical_alignment.Fill
                                                            Padding:=margin{Left:=PetsLineBoxPadding,Top:=PetsLineBoxPadding,Right:=PetsLineBoxPadding,Bottom:=PetsLineBoxPadding}
                                                            Widget:=HealthLine_Player_Slot4
                                                        overlay_slot:
                                                            HorizontalAlignment:=horizontal_alignment.Left
                                                            VerticalAlignment:=vertical_alignment.Center
                                                            Widget:=overlay:
                                                                Slots:=array:
                                                                    overlay_slot:
                                                                        HorizontalAlignment:=horizontal_alignment.Center
                                                                        VerticalAlignment:=vertical_alignment.Center
                                                                        Padding:=margin{Left:=7.0,Top:=PetsLineBoxPadding,Right:=PetsLineBoxPadding,Bottom:=PetsLineBoxPadding}
                                                                        Widget:=PetImage_Player_Slot4
                                                        overlay_slot:
                                                            HorizontalAlignment:=horizontal_alignment.Left
                                                            VerticalAlignment:=vertical_alignment.Center
                                                            Padding:=margin{Left:=TextPetsBoxPadding}
                                                            Widget:=stack_box:
                                                                Orientation:=orientation.Vertical
                                                                    Slots:=array:
                                                                        stack_box_slot: #Class
                                                                            HorizontalAlignment:=horizontal_alignment.Left
                                                                            VerticalAlignment:=vertical_alignment.Center
                                                                            Padding:=margin{Left:=TextPetsBoxPadding}
                                                                            Widget:=stack_box:
                                                                                Orientation:=orientation.Horizontal
                                                                                Slots:=array:
                                                                                    stack_box_slot: 
                                                                                        Padding:=margin{Left:=2.5,Right:=2.5}
                                                                                        Widget:=Class_Player_Slot4
                                                                                    stack_box_slot: 
                                                                                        Padding:=margin{Left:=2.5, Right:=1.0}
                                                                                        Widget:=PackA_Player_Box4
                                                                                    stack_box_slot: 
                                                                                        Padding:=margin{Right:=2.5}
                                                                                        Widget:=PackB_Player_Box4
                                                                        stack_box_slot: #Purity
                                                                            HorizontalAlignment:=horizontal_alignment.Left
                                                                            VerticalAlignment:=vertical_alignment.Center
                                                                            Padding:=margin{Left:=TextPetsBoxPadding}
                                                                            Widget:=Purity_Player_Slot4
                                                                        stack_box_slot: #Health
                                                                            HorizontalAlignment:=horizontal_alignment.Left
                                                                            VerticalAlignment:=vertical_alignment.Center
                                                                            Padding:=margin{Left:=TextPetsBoxPadding}
                                                                            Widget:=Health_Player_Slot4
                                            stack_box_slot: #Slot 5
                                                Padding:=margin{Top:=ResourcersRowTopPadding,Bottom:=ResourcersRowBottomPadding}
                                                HorizontalAlignment:=horizontal_alignment.Fill
                                                VerticalAlignment:=vertical_alignment.Fill
                                                Widget:=overlay:
                                                    Slots:=array:
                                                        overlay_slot:
                                                            HorizontalAlignment:=horizontal_alignment.Fill
                                                            VerticalAlignment:=vertical_alignment.Fill
                                                            Widget:=EmptyImage_Player_Slot5
                                                        overlay_slot:
                                                            HorizontalAlignment:=horizontal_alignment.Fill
                                                            VerticalAlignment:=vertical_alignment.Fill
                                                            Widget:= texture_block{DefaultImage:=Images.Img_LineBackground}
                                                        overlay_slot:
                                                            HorizontalAlignment:=horizontal_alignment.Fill
                                                            VerticalAlignment:=vertical_alignment.Fill
                                                            Padding:=margin{Left:=PetsLineBoxPadding,Top:=PetsLineBoxPadding,Right:=PetsLineBoxPadding,Bottom:=PetsLineBoxPadding}
                                                            Widget:= texture_block{DefaultImage:=Images.Img_LineBackground}
                                                        overlay_slot:
                                                            HorizontalAlignment:=horizontal_alignment.Left
                                                            VerticalAlignment:=vertical_alignment.Fill
                                                            Padding:=margin{Left:=PetsLineBoxPadding,Top:=PetsLineBoxPadding,Right:=PetsLineBoxPadding,Bottom:=PetsLineBoxPadding}
                                                            Widget:=HealthLine_Player_Slot5
                                                        overlay_slot:
                                                            HorizontalAlignment:=horizontal_alignment.Left
                                                            VerticalAlignment:=vertical_alignment.Center
                                                            Widget:=overlay:
                                                                Slots:=array:
                                                                    overlay_slot:
                                                                        HorizontalAlignment:=horizontal_alignment.Center
                                                                        VerticalAlignment:=vertical_alignment.Center
                                                                        Padding:=margin{Left:=7.0,Top:=PetsLineBoxPadding,Right:=PetsLineBoxPadding,Bottom:=PetsLineBoxPadding}
                                                                        Widget:=PetImage_Player_Slot5
                                                        overlay_slot:
                                                            HorizontalAlignment:=horizontal_alignment.Left
                                                            VerticalAlignment:=vertical_alignment.Center
                                                            Padding:=margin{Left:=TextPetsBoxPadding}
                                                            Widget:=stack_box:
                                                                Orientation:=orientation.Vertical
                                                                    Slots:=array:
                                                                        stack_box_slot: #Class
                                                                            HorizontalAlignment:=horizontal_alignment.Left
                                                                            VerticalAlignment:=vertical_alignment.Center
                                                                            Padding:=margin{Left:=TextPetsBoxPadding}
                                                                            Widget:=stack_box:
                                                                                Orientation:=orientation.Horizontal
                                                                                Slots:=array:
                                                                                    stack_box_slot: 
                                                                                        Padding:=margin{Left:=2.5,Right:=2.5}
                                                                                        Widget:=Class_Player_Slot5
                                                                                    stack_box_slot: 
                                                                                        Padding:=margin{Left:=2.5, Right:=1.0}
                                                                                        Widget:=PackA_Player_Box5
                                                                                    stack_box_slot: 
                                                                                        Padding:=margin{Right:=2.5}
                                                                                        Widget:=PackB_Player_Box5
                                                                        stack_box_slot: #Purity
                                                                            HorizontalAlignment:=horizontal_alignment.Left
                                                                            VerticalAlignment:=vertical_alignment.Center
                                                                            Padding:=margin{Left:=TextPetsBoxPadding}
                                                                            Widget:=Purity_Player_Slot5
                                                                        stack_box_slot: #Health
                                                                            HorizontalAlignment:=horizontal_alignment.Left
                                                                            VerticalAlignment:=vertical_alignment.Center
                                                                            Padding:=margin{Left:=TextPetsBoxPadding}
                                                                            Widget:=Health_Player_Slot5
                                                                
                                

                canvas_slot:
                    Anchors:=anchors{Minimum:=vector2{X:=1.0,Y:=0.5},Maximum:=vector2{X:=1.0,Y:=0.5}}
                    ZOrder:=0
                    SizeToContent:=true
                    Offsets:=margin{Left:=-5.0,Top:=120.0,Right:=0.0,Bottom:=0.0} #PositionX, PositionY, SizeX, SizeY
                    Alignment:=vector2{X:=1.0,Y:=0.5}
                    Widget:=overlay:
                        Slots:=array:
                            overlay_slot:
                                HorizontalAlignment:=horizontal_alignment.Fill
                                VerticalAlignment:=vertical_alignment.Fill
                                Widget:=stack_box:
                                    Orientation:=orientation.Vertical
                                        Slots:=array:
                                            stack_box_slot: #Slot 1
                                                Padding:=margin{Top:=ResourcersRowTopPadding,Bottom:=ResourcersRowBottomPadding}
                                                HorizontalAlignment:=horizontal_alignment.Fill
                                                VerticalAlignment:=vertical_alignment.Fill
                                                Widget:=overlay:
                                                    Slots:=array:
                                                        overlay_slot:
                                                            HorizontalAlignment:=horizontal_alignment.Fill
                                                            VerticalAlignment:=vertical_alignment.Fill
                                                            Widget:=EmptyImage_AI_Slot1
                                                        overlay_slot:
                                                            HorizontalAlignment:=horizontal_alignment.Fill
                                                            VerticalAlignment:=vertical_alignment.Fill
                                                            Widget:= texture_block{DefaultImage:=Images.Img_LineBackground}
                                                        overlay_slot:
                                                            HorizontalAlignment:=horizontal_alignment.Fill
                                                            VerticalAlignment:=vertical_alignment.Fill
                                                            Padding:=margin{Left:=PetsLineBoxPadding,Top:=PetsLineBoxPadding,Right:=PetsLineBoxPadding,Bottom:=PetsLineBoxPadding}
                                                            Widget:= texture_block{DefaultImage:=Images.Img_LineBackground}
                                                        overlay_slot:
                                                            HorizontalAlignment:=horizontal_alignment.Left
                                                            VerticalAlignment:=vertical_alignment.Fill
                                                            Padding:=margin{Left:=PetsLineBoxPadding,Top:=PetsLineBoxPadding,Right:=PetsLineBoxPadding,Bottom:=PetsLineBoxPadding}
                                                            Widget:=HealthLine_AI_Slot1
                                                        overlay_slot:
                                                            HorizontalAlignment:=horizontal_alignment.Left
                                                            VerticalAlignment:=vertical_alignment.Center
                                                            Widget:=overlay:
                                                                Slots:=array:
                                                                    overlay_slot:
                                                                        HorizontalAlignment:=horizontal_alignment.Center
                                                                        VerticalAlignment:=vertical_alignment.Center
                                                                        Padding:=margin{Left:=7.0,Top:=PetsLineBoxPadding,Right:=PetsLineBoxPadding,Bottom:=PetsLineBoxPadding}
                                                                        Widget:=PetImage_AI_Slot1
                                                        overlay_slot:
                                                            HorizontalAlignment:=horizontal_alignment.Left
                                                            VerticalAlignment:=vertical_alignment.Center
                                                            Padding:=margin{Left:=TextPetsBoxPadding}
                                                            Widget:=stack_box:
                                                                Orientation:=orientation.Vertical
                                                                    Slots:=array:
                                                                        stack_box_slot: #Class
                                                                            HorizontalAlignment:=horizontal_alignment.Left
                                                                            VerticalAlignment:=vertical_alignment.Center
                                                                            Padding:=margin{Left:=TextPetsBoxPadding}
                                                                            Widget:=stack_box:
                                                                                Orientation:=orientation.Horizontal
                                                                                Slots:=array:
                                                                                    stack_box_slot: 
                                                                                        Padding:=margin{Left:=2.5,Right:=2.5}
                                                                                        Widget:=Class_AI_Slot1
                                                                                    stack_box_slot: 
                                                                                        Padding:=margin{Left:=2.5, Right:=1.0}
                                                                                        Widget:=PackA_AI_Box1
                                                                                    stack_box_slot: 
                                                                                        Padding:=margin{Right:=2.5}
                                                                                        Widget:=PackB_AI_Box1
                                                                        stack_box_slot: #Purity
                                                                            HorizontalAlignment:=horizontal_alignment.Left
                                                                            VerticalAlignment:=vertical_alignment.Center
                                                                            Padding:=margin{Left:=TextPetsBoxPadding}
                                                                            Widget:=Purity_AI_Slot1
                                                                        stack_box_slot: #Health
                                                                            HorizontalAlignment:=horizontal_alignment.Left
                                                                            VerticalAlignment:=vertical_alignment.Center
                                                                            Padding:=margin{Left:=TextPetsBoxPadding}
                                                                            Widget:=Health_AI_Slot1
                                            stack_box_slot: #Slot 2
                                                Padding:=margin{Top:=ResourcersRowTopPadding,Bottom:=ResourcersRowBottomPadding}
                                                HorizontalAlignment:=horizontal_alignment.Fill
                                                VerticalAlignment:=vertical_alignment.Fill
                                                Widget:=overlay:
                                                    Slots:=array:
                                                        overlay_slot:
                                                            HorizontalAlignment:=horizontal_alignment.Fill
                                                            VerticalAlignment:=vertical_alignment.Fill
                                                            Widget:=EmptyImage_AI_Slot2
                                                        overlay_slot:
                                                            HorizontalAlignment:=horizontal_alignment.Fill
                                                            VerticalAlignment:=vertical_alignment.Fill
                                                            Widget:= texture_block{DefaultImage:=Images.Img_LineBackground}
                                                        overlay_slot:
                                                            HorizontalAlignment:=horizontal_alignment.Fill
                                                            VerticalAlignment:=vertical_alignment.Fill
                                                            Padding:=margin{Left:=PetsLineBoxPadding,Top:=PetsLineBoxPadding,Right:=PetsLineBoxPadding,Bottom:=PetsLineBoxPadding}
                                                            Widget:= texture_block{DefaultImage:=Images.Img_LineBackground}
                                                        overlay_slot:
                                                            HorizontalAlignment:=horizontal_alignment.Left
                                                            VerticalAlignment:=vertical_alignment.Fill
                                                            Padding:=margin{Left:=PetsLineBoxPadding,Top:=PetsLineBoxPadding,Right:=PetsLineBoxPadding,Bottom:=PetsLineBoxPadding}
                                                            Widget:=HealthLine_AI_Slot2
                                                        overlay_slot:
                                                            HorizontalAlignment:=horizontal_alignment.Left
                                                            VerticalAlignment:=vertical_alignment.Center
                                                            Widget:=overlay:
                                                                Slots:=array:
                                                                    overlay_slot:
                                                                        HorizontalAlignment:=horizontal_alignment.Center
                                                                        VerticalAlignment:=vertical_alignment.Center
                                                                        Padding:=margin{Left:=7.0,Top:=PetsLineBoxPadding,Right:=PetsLineBoxPadding,Bottom:=PetsLineBoxPadding}
                                                                        Widget:=PetImage_AI_Slot2
                                                        overlay_slot:
                                                            HorizontalAlignment:=horizontal_alignment.Left
                                                            VerticalAlignment:=vertical_alignment.Center
                                                            Padding:=margin{Left:=TextPetsBoxPadding}
                                                            Widget:=stack_box:
                                                                Orientation:=orientation.Vertical
                                                                    Slots:=array:
                                                                        stack_box_slot: #Class
                                                                            HorizontalAlignment:=horizontal_alignment.Left
                                                                            VerticalAlignment:=vertical_alignment.Center
                                                                            Padding:=margin{Left:=TextPetsBoxPadding}
                                                                            Widget:=stack_box:
                                                                                Orientation:=orientation.Horizontal
                                                                                Slots:=array:
                                                                                    stack_box_slot: 
                                                                                        Padding:=margin{Left:=2.5,Right:=2.5}
                                                                                        Widget:=Class_AI_Slot2
                                                                                    stack_box_slot: 
                                                                                        Padding:=margin{Left:=2.5, Right:=1.0}
                                                                                        Widget:=PackA_AI_Box2
                                                                                    stack_box_slot: 
                                                                                        Padding:=margin{Right:=2.5}
                                                                                        Widget:=PackB_AI_Box2
                                                                        stack_box_slot: #Purity
                                                                            HorizontalAlignment:=horizontal_alignment.Left
                                                                            VerticalAlignment:=vertical_alignment.Center
                                                                            Padding:=margin{Left:=TextPetsBoxPadding}
                                                                            Widget:=Purity_AI_Slot2
                                                                        stack_box_slot: #Health
                                                                            HorizontalAlignment:=horizontal_alignment.Left
                                                                            VerticalAlignment:=vertical_alignment.Center
                                                                            Padding:=margin{Left:=TextPetsBoxPadding}
                                                                            Widget:=Health_AI_Slot2
                                            stack_box_slot: #Slot 3
                                                Padding:=margin{Top:=ResourcersRowTopPadding,Bottom:=ResourcersRowBottomPadding}
                                                HorizontalAlignment:=horizontal_alignment.Fill
                                                VerticalAlignment:=vertical_alignment.Fill
                                                Widget:=overlay:
                                                    Slots:=array:
                                                        overlay_slot:
                                                            HorizontalAlignment:=horizontal_alignment.Fill
                                                            VerticalAlignment:=vertical_alignment.Fill
                                                            Widget:=EmptyImage_AI_Slot3
                                                        overlay_slot:
                                                            HorizontalAlignment:=horizontal_alignment.Fill
                                                            VerticalAlignment:=vertical_alignment.Fill
                                                            Widget:= texture_block{DefaultImage:=Images.Img_LineBackground}
                                                        overlay_slot:
                                                            HorizontalAlignment:=horizontal_alignment.Fill
                                                            VerticalAlignment:=vertical_alignment.Fill
                                                            Padding:=margin{Left:=PetsLineBoxPadding,Top:=PetsLineBoxPadding,Right:=PetsLineBoxPadding,Bottom:=PetsLineBoxPadding}
                                                            Widget:= texture_block{DefaultImage:=Images.Img_LineBackground}
                                                        overlay_slot:
                                                            HorizontalAlignment:=horizontal_alignment.Left
                                                            VerticalAlignment:=vertical_alignment.Fill
                                                            Padding:=margin{Left:=PetsLineBoxPadding,Top:=PetsLineBoxPadding,Right:=PetsLineBoxPadding,Bottom:=PetsLineBoxPadding}
                                                            Widget:=HealthLine_AI_Slot3
                                                        overlay_slot:
                                                            HorizontalAlignment:=horizontal_alignment.Left
                                                            VerticalAlignment:=vertical_alignment.Center
                                                            Widget:=overlay:
                                                                Slots:=array:
                                                                    overlay_slot:
                                                                        HorizontalAlignment:=horizontal_alignment.Center
                                                                        VerticalAlignment:=vertical_alignment.Center
                                                                        Padding:=margin{Left:=7.0,Top:=PetsLineBoxPadding,Right:=PetsLineBoxPadding,Bottom:=PetsLineBoxPadding}
                                                                        Widget:=PetImage_AI_Slot3
                                                        overlay_slot:
                                                            HorizontalAlignment:=horizontal_alignment.Left
                                                            VerticalAlignment:=vertical_alignment.Center
                                                            Padding:=margin{Left:=TextPetsBoxPadding}
                                                            Widget:=stack_box:
                                                                Orientation:=orientation.Vertical
                                                                    Slots:=array:
                                                                        stack_box_slot: #Class
                                                                            HorizontalAlignment:=horizontal_alignment.Left
                                                                            VerticalAlignment:=vertical_alignment.Center
                                                                            Padding:=margin{Left:=TextPetsBoxPadding}
                                                                            Widget:=stack_box:
                                                                                Orientation:=orientation.Horizontal
                                                                                Slots:=array:
                                                                                    stack_box_slot: 
                                                                                        Padding:=margin{Left:=2.5,Right:=2.5}
                                                                                        Widget:=Class_AI_Slot3
                                                                                    stack_box_slot: 
                                                                                        Padding:=margin{Left:=2.5, Right:=1.0}
                                                                                        Widget:=PackA_AI_Box3
                                                                                    stack_box_slot: 
                                                                                        Padding:=margin{Right:=2.5}
                                                                                        Widget:=PackB_AI_Box3
                                                                        stack_box_slot: #Purity
                                                                            HorizontalAlignment:=horizontal_alignment.Left
                                                                            VerticalAlignment:=vertical_alignment.Center
                                                                            Padding:=margin{Left:=TextPetsBoxPadding}
                                                                            Widget:=Purity_AI_Slot3
                                                                        stack_box_slot: #Health
                                                                            HorizontalAlignment:=horizontal_alignment.Left
                                                                            VerticalAlignment:=vertical_alignment.Center
                                                                            Padding:=margin{Left:=TextPetsBoxPadding}
                                                                            Widget:=Health_AI_Slot3
                                            stack_box_slot: #Slot 4
                                                Padding:=margin{Top:=ResourcersRowTopPadding,Bottom:=ResourcersRowBottomPadding}
                                                HorizontalAlignment:=horizontal_alignment.Fill
                                                VerticalAlignment:=vertical_alignment.Fill
                                                Widget:=overlay:
                                                    Slots:=array:
                                                        overlay_slot:
                                                            HorizontalAlignment:=horizontal_alignment.Fill
                                                            VerticalAlignment:=vertical_alignment.Fill
                                                            Widget:=EmptyImage_AI_Slot4
                                                        overlay_slot:
                                                            HorizontalAlignment:=horizontal_alignment.Fill
                                                            VerticalAlignment:=vertical_alignment.Fill
                                                            Widget:= texture_block{DefaultImage:=Images.Img_LineBackground}
                                                        overlay_slot:
                                                            HorizontalAlignment:=horizontal_alignment.Fill
                                                            VerticalAlignment:=vertical_alignment.Fill
                                                            Padding:=margin{Left:=PetsLineBoxPadding,Top:=PetsLineBoxPadding,Right:=PetsLineBoxPadding,Bottom:=PetsLineBoxPadding}
                                                            Widget:= texture_block{DefaultImage:=Images.Img_LineBackground}
                                                        overlay_slot:
                                                            HorizontalAlignment:=horizontal_alignment.Left
                                                            VerticalAlignment:=vertical_alignment.Fill
                                                            Padding:=margin{Left:=PetsLineBoxPadding,Top:=PetsLineBoxPadding,Right:=PetsLineBoxPadding,Bottom:=PetsLineBoxPadding}
                                                            Widget:=HealthLine_AI_Slot4
                                                        overlay_slot:
                                                            HorizontalAlignment:=horizontal_alignment.Left
                                                            VerticalAlignment:=vertical_alignment.Center
                                                            Widget:=overlay:
                                                                Slots:=array:
                                                                    overlay_slot:
                                                                        HorizontalAlignment:=horizontal_alignment.Center
                                                                        VerticalAlignment:=vertical_alignment.Center
                                                                        Padding:=margin{Left:=7.0,Top:=PetsLineBoxPadding,Right:=PetsLineBoxPadding,Bottom:=PetsLineBoxPadding}
                                                                        Widget:=PetImage_AI_Slot4
                                                        overlay_slot:
                                                            HorizontalAlignment:=horizontal_alignment.Left
                                                            VerticalAlignment:=vertical_alignment.Center
                                                            Padding:=margin{Left:=TextPetsBoxPadding}
                                                            Widget:=stack_box:
                                                                Orientation:=orientation.Vertical
                                                                    Slots:=array:
                                                                        stack_box_slot: #Class
                                                                            HorizontalAlignment:=horizontal_alignment.Left
                                                                            VerticalAlignment:=vertical_alignment.Center
                                                                            Padding:=margin{Left:=TextPetsBoxPadding}
                                                                            Widget:=stack_box:
                                                                                Orientation:=orientation.Horizontal
                                                                                Slots:=array:
                                                                                    stack_box_slot: 
                                                                                        Padding:=margin{Left:=2.5,Right:=2.5}
                                                                                        Widget:=Class_AI_Slot4
                                                                                    stack_box_slot: 
                                                                                        Padding:=margin{Left:=2.5, Right:=1.0}
                                                                                        Widget:=PackA_AI_Box4
                                                                                    stack_box_slot: 
                                                                                        Padding:=margin{Right:=2.5}
                                                                                        Widget:=PackB_AI_Box4
                                                                        stack_box_slot: #Purity
                                                                            HorizontalAlignment:=horizontal_alignment.Left
                                                                            VerticalAlignment:=vertical_alignment.Center
                                                                            Padding:=margin{Left:=TextPetsBoxPadding}
                                                                            Widget:=Purity_AI_Slot4
                                                                        stack_box_slot: #Health
                                                                            HorizontalAlignment:=horizontal_alignment.Left
                                                                            VerticalAlignment:=vertical_alignment.Center
                                                                            Padding:=margin{Left:=TextPetsBoxPadding}
                                                                            Widget:=Health_AI_Slot4
                                            stack_box_slot: #Slot 5
                                                Padding:=margin{Top:=ResourcersRowTopPadding,Bottom:=ResourcersRowBottomPadding}
                                                HorizontalAlignment:=horizontal_alignment.Fill
                                                VerticalAlignment:=vertical_alignment.Fill
                                                Widget:=overlay:
                                                    Slots:=array:
                                                        overlay_slot:
                                                            HorizontalAlignment:=horizontal_alignment.Fill
                                                            VerticalAlignment:=vertical_alignment.Fill
                                                            Widget:=EmptyImage_AI_Slot5
                                                        overlay_slot:
                                                            HorizontalAlignment:=horizontal_alignment.Fill
                                                            VerticalAlignment:=vertical_alignment.Fill
                                                            Widget:= texture_block{DefaultImage:=Images.Img_LineBackground}
                                                        overlay_slot:
                                                            HorizontalAlignment:=horizontal_alignment.Fill
                                                            VerticalAlignment:=vertical_alignment.Fill
                                                            Padding:=margin{Left:=PetsLineBoxPadding,Top:=PetsLineBoxPadding,Right:=PetsLineBoxPadding,Bottom:=PetsLineBoxPadding}
                                                            Widget:= texture_block{DefaultImage:=Images.Img_LineBackground}
                                                        overlay_slot:
                                                            HorizontalAlignment:=horizontal_alignment.Left
                                                            VerticalAlignment:=vertical_alignment.Fill
                                                            Padding:=margin{Left:=PetsLineBoxPadding,Top:=PetsLineBoxPadding,Right:=PetsLineBoxPadding,Bottom:=PetsLineBoxPadding}
                                                            Widget:=HealthLine_AI_Slot5
                                                        overlay_slot:
                                                            HorizontalAlignment:=horizontal_alignment.Left
                                                            VerticalAlignment:=vertical_alignment.Center
                                                            Widget:=overlay:
                                                                Slots:=array:
                                                                    overlay_slot:
                                                                        HorizontalAlignment:=horizontal_alignment.Center
                                                                        VerticalAlignment:=vertical_alignment.Center
                                                                        Padding:=margin{Left:=7.0,Top:=PetsLineBoxPadding,Right:=PetsLineBoxPadding,Bottom:=PetsLineBoxPadding}
                                                                        Widget:=PetImage_AI_Slot5
                                                        overlay_slot:
                                                            HorizontalAlignment:=horizontal_alignment.Left
                                                            VerticalAlignment:=vertical_alignment.Center
                                                            Padding:=margin{Left:=TextPetsBoxPadding}
                                                            Widget:=stack_box:
                                                                Orientation:=orientation.Vertical
                                                                    Slots:=array:
                                                                        stack_box_slot: #Class
                                                                            HorizontalAlignment:=horizontal_alignment.Left
                                                                            VerticalAlignment:=vertical_alignment.Center
                                                                            Padding:=margin{Left:=TextPetsBoxPadding}
                                                                            Widget:=stack_box:
                                                                                Orientation:=orientation.Horizontal
                                                                                Slots:=array:
                                                                                    stack_box_slot: 
                                                                                        Padding:=margin{Left:=2.5,Right:=2.5}
                                                                                        Widget:=Class_AI_Slot5
                                                                                    stack_box_slot: 
                                                                                        Padding:=margin{Left:=2.5, Right:=1.0}
                                                                                        Widget:=PackA_AI_Box5
                                                                                    stack_box_slot: 
                                                                                        Padding:=margin{Right:=2.5}
                                                                                        Widget:=PackB_AI_Box5
                                                                        stack_box_slot: #Purity
                                                                            HorizontalAlignment:=horizontal_alignment.Left
                                                                            VerticalAlignment:=vertical_alignment.Center
                                                                            Padding:=margin{Left:=TextPetsBoxPadding}
                                                                            Widget:=Purity_AI_Slot5
                                                                        stack_box_slot: #Health
                                                                            HorizontalAlignment:=horizontal_alignment.Left
                                                                            VerticalAlignment:=vertical_alignment.Center
                                                                            Padding:=margin{Left:=TextPetsBoxPadding}
                                                                            Widget:=Health_AI_Slot5

MakeHexaMinigameUI<constructor>(
    InHexaMinigameManager : hexa_minigame_manager)<transacts> := hexa_minigame_ui :
        HexaMinigameManager := InHexaMinigameManager
                                                                                                                