# fill this file with your verse-script
using { /UnrealEngine.com/Temporary/UI }
using { /UnrealEngine.com/Temporary/SpatialMath }
using { /Fortnite.com/UI }
using { /Verse.org/Simulation }
using { /Verse.org/Colors/NamedColors }
using { /Fortnite.com/Characters }

# A Verse-authored creative device that can be placed in a level
language_ui := class():

    BakeryMachinesManager : bakery_machines_manager

    #Class settings
    var MaybeCanvas : ?canvas = false
    var UIIsDisplayed :logic = false

    var ConfirmationToAll : int = 0

    ShowQuestTracker_Button : button_regular = button_regular{}
    LanguageInfo : text_block = text_block{DefaultTextColor:=White, DefaultShadowOffset:=ShadowOffset, DefaultShadowOpacity := 1.0}
    Language0_Button : button_regular = button_regular{}
    Language1_Button : button_regular = button_regular{}
    Language2_Button : button_regular = button_regular{}
    Language3_Button : button_regular = button_regular{}
    Cancel_Button : button_loud = button_loud{}

    var TimeElapsed : int = 0
    MaxTimeElapsed : int = 15

    InitUI() : void =

        Canvas := MakeOptionsMenuCanvas()
        set MaybeCanvas = option{Canvas}

        ShowQuestTracker_Button.SetText(BasicMessage("Show quest tracker"))
        Language0_Button.SetText(BasicMessage("English"))
        Language1_Button.SetText(BasicMessage("Español"))
        Language2_Button.SetText(BasicMessage("Français"))
        Language3_Button.SetText(BasicMessage("Português"))
        LanguageInfo.SetText(BasicMessage("Note: Not all game content can be translated yet"))
        Cancel_Button.SetText(BasicMessage("Close "))

        Subscribers()

    Subscribers() : void =
        Language0_Button.OnClick().Subscribe(SwitchToEnglish) 
        Language1_Button.OnClick().Subscribe(SwitchToSpanish) 
        Language2_Button.OnClick().Subscribe(SwitchToFrench) 
        Language3_Button.OnClick().Subscribe(SwitchToPortuguese) 
        Cancel_Button.OnClick().Subscribe(Remove_Canvas)
        
        
    SwitchToEnglish_N() : void =
        if(Agent := BakeryMachinesManager.MaybeAgentOwner?, Player := player[Agent], PlayerData := BakeryMachinesManager.PlayerDataCache):
            BakeryMachinesManager.SetNewPlayerDataCache(
                BakeryMachinesManager.PlayersDataManager.SetLanguage(BakeryMachinesManager.PlayerDataCache, ENGLISH))
            BakeryMachinesManager.GameManager.DisplayCharacterMessage(
                Agent, I_CHARACTER, "Language changed to English", ?Sound := MESSAGE_SOUND)
        UpdateUI()

    SwitchToSpanish_N() : void =
        if(Agent := BakeryMachinesManager.MaybeAgentOwner?, Player := player[Agent], PlayerData := BakeryMachinesManager.PlayerDataCache):
            BakeryMachinesManager.SetNewPlayerDataCache(
                BakeryMachinesManager.PlayersDataManager.SetLanguage(BakeryMachinesManager.PlayerDataCache, SPANISH))
            BakeryMachinesManager.GameManager.DisplayCharacterMessage(
                Agent, I_CHARACTER, "Idioma cambiado a Español", ?Sound := MESSAGE_SOUND)
        UpdateUI()
    
    SwitchToFrench_N() : void =
            if(Agent := BakeryMachinesManager.MaybeAgentOwner?, Player := player[Agent], PlayerData := BakeryMachinesManager.PlayerDataCache):
                BakeryMachinesManager.SetNewPlayerDataCache(
                    BakeryMachinesManager.PlayersDataManager.SetLanguage(BakeryMachinesManager.PlayerDataCache, FRENCH))
                BakeryMachinesManager.GameManager.DisplayCharacterMessage(
                    Agent, I_CHARACTER, "Langue changée en Français", ?Sound := MESSAGE_SOUND)
            UpdateUI()
    
    SwitchToPortuguese_N() : void =
            if(Agent := BakeryMachinesManager.MaybeAgentOwner?, Player := player[Agent], PlayerData := BakeryMachinesManager.PlayerDataCache):
                BakeryMachinesManager.SetNewPlayerDataCache(
                    BakeryMachinesManager.PlayersDataManager.SetLanguage(BakeryMachinesManager.PlayerDataCache, PORTUGUESE))
                BakeryMachinesManager.GameManager.DisplayCharacterMessage(
                    Agent, I_CHARACTER, "Idioma alterado para Português", ?Sound := MESSAGE_SOUND)
            UpdateUI()


    SwitchToEnglish(Message : widget_message) : void =
        if(Agent:=BakeryMachinesManager.MaybeAgentOwner?, Player := player[Agent], PlayerData:=BakeryMachinesManager.PlayerDataCache):
            BakeryMachinesManager.SetNewPlayerDataCache(BakeryMachinesManager.PlayersDataManager.SetLanguage(BakeryMachinesManager.PlayerDataCache,ENGLISH))
            BakeryMachinesManager.GameManager.DisplayCharacterMessage(Agent,I_CHARACTER,"Language changed to English", ?Sound:=MESSAGE_SOUND)
        UpdateUI()
        Remove_Canvas(Message)

    SwitchToSpanish(Message : widget_message) : void =
        if(Agent:=BakeryMachinesManager.MaybeAgentOwner?, Player := player[Agent], PlayerData:=BakeryMachinesManager.PlayerDataCache):
            BakeryMachinesManager.SetNewPlayerDataCache(BakeryMachinesManager.PlayersDataManager.SetLanguage(BakeryMachinesManager.PlayerDataCache,SPANISH))
            BakeryMachinesManager.GameManager.DisplayCharacterMessage(Agent,I_CHARACTER,"Idioma cambiado a Español", ?Sound:=MESSAGE_SOUND)
        UpdateUI()
        Remove_Canvas(Message)

    SwitchToFrench(Message : widget_message) : void =
        if(Agent:=BakeryMachinesManager.MaybeAgentOwner?, Player := player[Agent], PlayerData:=BakeryMachinesManager.PlayerDataCache):
            BakeryMachinesManager.SetNewPlayerDataCache(BakeryMachinesManager.PlayersDataManager.SetLanguage(BakeryMachinesManager.PlayerDataCache,FRENCH))
            BakeryMachinesManager.GameManager.DisplayCharacterMessage(Agent,I_CHARACTER,"Langue changée en Français", ?Sound:=MESSAGE_SOUND)
        UpdateUI()
        Remove_Canvas(Message)
    
    SwitchToPortuguese(Message : widget_message) : void =
        if(Agent:=BakeryMachinesManager.MaybeAgentOwner?, Player := player[Agent], PlayerData:=BakeryMachinesManager.PlayerDataCache):
            BakeryMachinesManager.SetNewPlayerDataCache(BakeryMachinesManager.PlayersDataManager.SetLanguage(BakeryMachinesManager.PlayerDataCache,PORTUGUESE))
            BakeryMachinesManager.GameManager.DisplayCharacterMessage(Agent,I_CHARACTER,"Idioma alterado para Português", ?Sound:=MESSAGE_SOUND)
        UpdateUI()
        Remove_Canvas(Message)
            
    UpdateUI() : void =
        if(BakeryMachinesManager.PlayerDataCache.Language = ENGLISH):
            LanguageInfo.SetText(BasicMessage("Note: Some game content is not yet translatable."))
        
        if(BakeryMachinesManager.PlayerDataCache.Language = SPANISH):
            LanguageInfo.SetText(BasicMessage("Nota: Parte del contenido aún no es traducible."))
        
        if(BakeryMachinesManager.PlayerDataCache.Language = FRENCH):
            LanguageInfo.SetText(BasicMessage("Remarque : Certains contenus ne sont pas encore traduits."))
        
        if(BakeryMachinesManager.PlayerDataCache.Language = PORTUGUESE):
            LanguageInfo.SetText(BasicMessage("Nota: Parte do conteúdo ainda não está traduzido."))
        
    Remove_Canvas(Message : widget_message) : void =
        set UIIsDisplayed = false
        set TimeElapsed = MaxTimeElapsed
        if:
            Agent:=BakeryMachinesManager.MaybeAgentOwner?
            Player:=player[Agent]
            PlayerUI:=GetPlayerUI[Player]
            Canvas := MaybeCanvas?
            UKTutorial:=BakeryMachinesManager.MaybeUKTutorial?
        then:
            UKTutorial.LanguageSelectedEvent()
            PlayerUI.RemoveWidget(Canvas)
            RemoveDelay_OnRemove()

    Remove_CanvasWithoutMessage() : void =
        set UIIsDisplayed = false
        set TimeElapsed = MaxTimeElapsed
        if:
            Agent:=BakeryMachinesManager.MaybeAgentOwner?
            Player:=player[Agent]
            PlayerUI:=GetPlayerUI[Player]
            Canvas := MaybeCanvas?
            UKTutorial:=BakeryMachinesManager.MaybeUKTutorial?
        then:
            UKTutorial.LanguageSelectedEvent()
            PlayerUI.RemoveWidget(Canvas)
            RemoveDelay_OnRemove()

    RemoveDelay_OnAdd() : void =
        Cancel_Button.SetText(BasicMessage("CLOSE"))


    RemoveDelay_OnRemove() : void =
        Cancel_Button.SetText(BasicMessage("CLOSE "))

    StartTimeElapsed()<suspends> : void =
        loop:
            Log("TimeElapsed: {TimeElapsed}", LOG_LEVEL_DEBUG9)
            if(TimeElapsed > MaxTimeElapsed):
                ##Print("Closing the UI because 20 seconds has pased")
                Remove_CanvasWithoutMessage()
                break
            if(UIIsDisplayed = false):
                ##Print("Breaking Time Elapsed loop")
                break
            Sleep(1.0)
            set TimeElapsed = TimeElapsed + 1
            
    Display_Canvas() : void =
        if(not UIIsDisplayed?):
            set UIIsDisplayed = true
            set TimeElapsed = 0
            spawn: 
                StartTimeElapsed()
            set ConfirmationToAll = 0
            set UIIsDisplayed = true
            var Index : int = 0
            if:
                Agent:=BakeryMachinesManager.MaybeAgentOwner?
                PlayerUI:=GetPlayerUI[player[Agent]]
                Canvas := MaybeCanvas?
            then:
                PlayerUI.AddWidget(Canvas,player_ui_slot{ZOrder:=1, InputMode:=ui_input_mode.All})
                RemoveDelay_OnAdd()   
            UpdateUI()

    MakeOptionsMenuCanvas(): canvas=
        Canvas := canvas:
            Slots:=array:
                #Background image
                canvas_slot:
                    Anchors:=anchors{Minimum:=vector2{X:=0.5,Y:=0.5},Maximum:=vector2{X:=0.5,Y:=0.5}}
                    ZOrder:=0
                    Offsets:=margin{Left:=0.0,Top:=0.0,Right:=0.0,Bottom:=0.0} #PositionX, PositionY, SizeX, SizeY
                    Alignment:=vector2{X:=0.5,Y:=0.5}
                    Widget:=overlay:
                        Slots:=array:
                            
                            overlay_slot:
                                HorizontalAlignment:=horizontal_alignment.Fill
                                VerticalAlignment:=vertical_alignment.Fill
                                Widget:=texture_block{DefaultImage:=Images.BlackBackground}
                            overlay_slot:
                                HorizontalAlignment:=horizontal_alignment.Fill
                                VerticalAlignment:=vertical_alignment.Fill
                                Padding:=margin{Left:=SelectorBoxPadding,Right:=SelectorBoxPadding,Top:=SelectorBoxPadding,Bottom:=SelectorBoxPadding}
                                Widget:=texture_block{DefaultImage:=Images.BlackBackground}
                            overlay_slot: #Screw 0
                                Padding:=margin{Left:=3.5,Top:=3.5,Right:=3.5,Bottom:=3.5}
                                HorizontalAlignment:=horizontal_alignment.Left
                                VerticalAlignment:=vertical_alignment.Top
                                Widget:=texture_block{DefaultImage:=Images.Dialogue_Bolt, DefaultDesiredSize:=vector2{X:=5.0,Y:=5.0}}
                            overlay_slot: #Screw 1
                                Padding:=margin{Left:=3.5,Top:=3.5,Right:=3.5,Bottom:=3.5}
                                HorizontalAlignment:=horizontal_alignment.Right
                                VerticalAlignment:=vertical_alignment.Top
                                Widget:=texture_block{DefaultImage:=Images.Dialogue_Bolt, DefaultDesiredSize:=vector2{X:=5.0,Y:=5.0}}
                            overlay_slot: #Screw 2
                                Padding:=margin{Left:=3.5,Top:=3.5,Right:=3.5,Bottom:=3.5}
                                HorizontalAlignment:=horizontal_alignment.Left
                                VerticalAlignment:=vertical_alignment.Bottom
                                Widget:=texture_block{DefaultImage:=Images.Dialogue_Bolt, DefaultDesiredSize:=vector2{X:=5.0,Y:=5.0}}
                            overlay_slot: #Screw 3
                                Padding:=margin{Left:=3.5,Top:=3.5,Right:=3.5,Bottom:=3.5}
                                HorizontalAlignment:=horizontal_alignment.Right
                                VerticalAlignment:=vertical_alignment.Bottom
                                Widget:=texture_block{DefaultImage:=Images.Dialogue_Bolt, DefaultDesiredSize:=vector2{X:=5.0,Y:=5.0}}
                                <#
                            overlay_slot:
                                HorizontalAlignment:=horizontal_alignment.Fill
                                VerticalAlignment:=vertical_alignment.Fill
                                Widget:=texture_block{DefaultImage:=Images.BlackBackground,DefaultDesiredSize:=vector2{X:=5.0,Y:=5.0}}#>
                            overlay_slot:
                                Padding:=margin{Top:=10.0,Bottom:=10.0,Right:=10.0,Left:=10.0}
                                Widget:=stack_box:
                                    Orientation:=orientation.Vertical
                                        Slots:=array:
                                            stack_box_slot: 
                                                Widget:=overlay:
                                                    Slots:=array:
                                                        overlay_slot:
                                                            Padding:=margin{Top:=5.0,Bottom:=5.0,Right:=5.0,Left:=5.0}
                                                            Widget:=text_block{DefaultText:=BasicMessage("CHANGE LANGUAGE"), DefaultTextColor:=White, DefaultShadowOffset:=ShadowOffset, DefaultShadowOpacity := DefaultTextShadowOpacity}
                                            stack_box_slot: 
                                                Widget:=overlay:
                                                    Slots:=array:
                                                        overlay_slot:
                                                            Padding:=margin{Top:=5.0,Bottom:=5.0,Right:=5.0,Left:=5.0}
                                                            Widget:=LanguageInfo
                                            stack_box_slot: 
                                                Padding:=margin{Top:=5.0,Bottom:=5.0,Right:=5.0,Left:=5.0}
                                                Widget:=overlay:
                                                    Slots:=array:
                                                        overlay_slot:
                                                            Widget:=texture_block{DefaultImage:=Images.EmptyTexture, DefaultDesiredSize:=vector2{X:=500.0,Y:=80.0}}
                                                        overlay_slot:
                                                            HorizontalAlignment:=horizontal_alignment.Fill
                                                            VerticalAlignment:=vertical_alignment.Fill
                                                            Padding:=margin{Top:=15.0,Bottom:=5.0,Right:=5.0,Left:=5.0}
                                                            Widget:=Language0_Button
                                            stack_box_slot: 
                                                Padding:=margin{Top:=5.0,Bottom:=5.0,Right:=5.0,Left:=5.0}
                                                Widget:=overlay:
                                                    Slots:=array:
                                                        overlay_slot:
                                                            Widget:=texture_block{DefaultImage:=Images.EmptyTexture, DefaultDesiredSize:=vector2{X:=500.0,Y:=80.0}}
                                                        overlay_slot:
                                                            HorizontalAlignment:=horizontal_alignment.Fill
                                                            VerticalAlignment:=vertical_alignment.Fill
                                                            Padding:=margin{Top:=15.0,Bottom:=15.0,Right:=5.0,Left:=5.0}
                                                            Widget:=Language1_Button
                                            stack_box_slot: 
                                                Padding:=margin{Top:=5.0,Bottom:=5.0,Right:=5.0,Left:=5.0}
                                                Widget:=overlay:
                                                    Slots:=array:
                                                        overlay_slot:
                                                            Widget:=texture_block{DefaultImage:=Images.EmptyTexture, DefaultDesiredSize:=vector2{X:=500.0,Y:=80.0}}
                                                        overlay_slot:
                                                            HorizontalAlignment:=horizontal_alignment.Fill
                                                            VerticalAlignment:=vertical_alignment.Fill
                                                            Padding:=margin{Top:=15.0,Bottom:=15.0,Right:=5.0,Left:=5.0}
                                                            Widget:=Language2_Button
                                            stack_box_slot: 
                                                Padding:=margin{Top:=5.0,Bottom:=5.0,Right:=5.0,Left:=5.0}
                                                Widget:=overlay:
                                                    Slots:=array:
                                                        overlay_slot:
                                                            Widget:=texture_block{DefaultImage:=Images.EmptyTexture, DefaultDesiredSize:=vector2{X:=500.0,Y:=80.0}}
                                                        overlay_slot:
                                                            HorizontalAlignment:=horizontal_alignment.Fill
                                                            VerticalAlignment:=vertical_alignment.Fill
                                                            Padding:=margin{Top:=15.0,Bottom:=15.0,Right:=5.0,Left:=5.0}
                                                            Widget:=Language3_Button
                                            stack_box_slot: 
                                                Padding:=margin{Top:=15.0,Bottom:=15.0,Right:=5.0,Left:=5.0}
                                                Widget:=overlay:
                                                    Slots:=array:
                                                        overlay_slot:
                                                            Widget:=texture_block{DefaultImage:=Images.EmptyTexture, DefaultDesiredSize:=vector2{X:=500.0,Y:=80.0}}
                                                        overlay_slot:
                                                            HorizontalAlignment:=horizontal_alignment.Fill
                                                            VerticalAlignment:=vertical_alignment.Fill
                                                            Padding:=margin{Top:=5.0,Bottom:=5.0,Right:=5.0,Left:=5.0}
                                                            Widget:=Cancel_Button

MakeLanguageUI<constructor>(InBakeryMachinesManager : bakery_machines_manager)<transacts> := language_ui :
    BakeryMachinesManager := InBakeryMachinesManager