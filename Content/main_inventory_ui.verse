# fill this file with your verse-script
using { /UnrealEngine.com/Temporary/UI }
using { /UnrealEngine.com/Temporary/SpatialMath }
using { /Fortnite.com/UI }
using { /Verse.org/Simulation }
using { /Verse.org/Colors/NamedColors }


# A Verse-authored creative device that can be placed in a level
main_inventory_ui := class():

    OptionsInfoPanelSize : vector2 = vector2{X:=137.0,Y:=64.0}

    InventoryBoxSizeX : float = 220.0
    InventoryBoxSizeY : float = 59.0

    InventoryLimitedBoxSizeX : float = 265.0
    InventoryLimitedBoxSizeY : float = 59.0

    var ColorBlink : int = 0

    HideSize : vector2 = vector2{X:=0.0,Y:=0.0}

    NotLimited_Background : vector2 = vector2{X:=205.0,Y:=59.0}
    NotLimited_Border : vector2 = vector2{X:=220.0,Y:=75.0}

    Limited_EmptyPack : vector2 = vector2{X:=219.0,Y:=65.0}
    Limited_Border : vector2 = vector2{X:=265.0,Y:=75.0}

    SmallBoltDesiredSize : vector2 = vector2{X:=5.0,Y:=5.0}

    InventoryIconPadding : float = 20.0
    InventoryTextPadding : float = 70.0

    ResourceSliderY : float = 59.0

    var IsQuestCompleted_BeingDisplayed : logic = false
    var IsOrderCompleted_BeingDisplayed : logic = false

    PetPerkBoxSizeX : float = 50.0
    PetPerkBoxSizeY : float = 50.0

    #OptionsMenu
    OptionsInfo_Texture : texture_block = texture_block{DefaultImage:=Images.OptionsDisplay, DefaultDesiredSize := vector2{X:=137.0,Y:=64.0}}

    #DamageEffect
    DamageEffect_Texture : texture_block = texture_block{DefaultImage:=Images.HealthBackground, DefaultDesiredSize := vector2{X:=0.0,Y:=0.0} ,DefaultTint:=Red}

    #Class settings
    var MaybeAgent : ?agent = false

    BakeryMachinesManager : bakery_machines_manager

    #PetPerksSliders
    var AllPetPerkSlider_Texture : []texture_block = array{}
    var AllPetPerk_Text : []text_block = array{}

    PetPerkSlider0_Texture : texture_block = texture_block{DefaultImage:=Images.Resourcers_InventoryUI_Part1}
    PetPerk0_Text : text_block = text_block{DefaultTextColor:=White, DefaultShadowOffset:=ShadowOffset, DefaultShadowOpacity := 1.0}

    PetPerkSlider1_Texture : texture_block = texture_block{DefaultImage:=Images.Resourcers_InventoryUI_Part1}
    PetPerk1_Text : text_block = text_block{DefaultTextColor:=White, DefaultShadowOffset:=ShadowOffset, DefaultShadowOpacity := 1.0}

    PetPerkSlider2_Texture : texture_block = texture_block{DefaultImage:=Images.Resourcers_InventoryUI_Part1}
    PetPerk2_Text : text_block = text_block{DefaultTextColor:=White, DefaultShadowOffset:=ShadowOffset, DefaultShadowOpacity := 1.0}

    #CharactersMessages
    var AllCharacterMaybeMessagesBatch : []?message = array{} #Localized message
    var AllCharacterStringBatch : []?string = array{}
    var AllCharactersBatch : []int = array{}
    var AllCharactersTimeBatch : []float = array{}
    var AllCharactersSounds : []int = array{}

    CharacterMessage_TextBlock : text_block = text_block{DefaultTextColor:=White, DefaultShadowOffset:=ShadowOffset, DefaultShadowOpacity := 1.0}
    CharacterImage_TextureBlock : texture_block = texture_block{DefaultImage:=Images.Mike_Icon, DefaultDesiredSize := vector2{X:=115.0,Y:=115.0}}
    CharacterMessage_TextureBlock : texture_block = texture_block{DefaultImage:=Images.DialogueBorder, DefaultDesiredSize := vector2{X:=750.0,Y:=150.0}}
    CharacterMessage_Slider : texture_block = texture_block{DefaultImage:=Images.EmptyTexture, DefaultDesiredSize:=vector2{X:=200.0,Y:=600.0}}
    
    CharacterMessage_TextBlockLoc : text_block = text_block{DefaultTextColor:=White, DefaultShadowOffset:=ShadowOffset, DefaultShadowOpacity := 1.0}
    CharacterImage_TextureBlockLoc : texture_block = texture_block{DefaultImage:=Images.Mike_Icon, DefaultDesiredSize := vector2{X:=115.0,Y:=115.0}}
    CharacterMessage_TextureBlockLoc : texture_block = texture_block{DefaultImage:=Images.DialogueBorder2, DefaultDesiredSize := vector2{X:=1250.0,Y:=150.0}}
    CharacterMessage_SliderLoc : texture_block = texture_block{DefaultImage:=Images.EmptyTexture, DefaultDesiredSize:=vector2{X:=200.0,Y:=600.0}}

    CharacterMessage_SliderDisplayedAmount : float = 144.0
    CharacterMessage_SliderNotDisplayedAmount : float = 660.0

    var CharacterMessageIsRunning : logic = false

    #Order
    #OrderDialogue_Slider : texture_block = texture_block{DefaultImage:=Images.EmptyTexture, DefaultDesiredSize:=vector2{X:=0.0,Y:=0.0}}
    OrderTitle_TextBlock : text_block = text_block{DefaultTextColor:=White, DefaultShadowOffset:=ShadowOffset, DefaultShadowOpacity := 1.0}
    OrderProduct_TextureBlock : texture_block = texture_block{DefaultImage:=Images.Bread_Icon, DefaultDesiredSize := vector2{X:=40.0,Y:=40.0}}
    OrderProgress_TextBlock : text_block = text_block{DefaultTextColor:=CurrentTextColor, DefaultShadowOffset:=ShadowOffset, DefaultShadowOpacity := 1.0}
    OrderReward_TextureBlock : texture_block = texture_block{DefaultImage:=Images.Coal_Icon, DefaultDesiredSize := vector2{X:=40.0,Y:=40.0}}
    OrderReward_TextBlock : text_block = text_block{DefaultTextColor:=CoalTextColor, DefaultShadowOffset:=ShadowOffset, DefaultShadowOpacity := 1.0}
    OrderRewardDescrip_TextBlock : text_block = text_block{DefaultTextColor:=White, DefaultShadowOffset:=ShadowOffset, DefaultShadowOpacity := 1.0}

    Order_Background0 : texture_block = texture_block{DefaultImage:=Images.Img_LineBackground, DefaultDesiredSize:=vector2{X:=1.0,Y:=1.0}}
    Order_Background1 : texture_block = texture_block{DefaultImage:=Images.Img_LineBackground, DefaultDesiredSize:=vector2{X:=1.0,Y:=1.0}}
    Order_Bolt0 : texture_block = texture_block{DefaultImage:=Images.Dialogue_Bolt, DefaultDesiredSize:=vector2{X:=5.0,Y:=5.0}}
    Order_Bolt1 : texture_block = texture_block{DefaultImage:=Images.Dialogue_Bolt, DefaultDesiredSize:=vector2{X:=5.0,Y:=5.0}}
    Order_Bolt2 : texture_block = texture_block{DefaultImage:=Images.Dialogue_Bolt, DefaultDesiredSize:=vector2{X:=5.0,Y:=5.0}}
    Order_Bolt3 : texture_block = texture_block{DefaultImage:=Images.Dialogue_Bolt, DefaultDesiredSize:=vector2{X:=5.0,Y:=5.0}}


    #Quest
    QuestTitle_TextBlock : text_block = text_block{DefaultTextColor:=White, DefaultShadowOffset:=ShadowOffset, DefaultShadowOpacity := 1.0}
    QuestIcon_TextureBlock : texture_block = texture_block{DefaultImage:=Images.EmptyTexture, DefaultDesiredSize := vector2{X:=40.0,Y:=40.0}}
    
    QuestDescription_TextBlock : text_block = text_block{DefaultTextColor:=CurrentTextColor, DefaultShadowOffset:=ShadowOffset, DefaultShadowOpacity := 1.0}
    QuestRewardDescrip_TextBlock : text_block = text_block{DefaultTextColor:=CurrentTextColor, DefaultShadowOffset:=ShadowOffset, DefaultShadowOpacity := 1.0}
    QuestProgress_TextBlock : text_block = text_block{DefaultTextColor:=UpgradeTextColor, DefaultShadowOffset:=ShadowOffset, DefaultShadowOpacity := 1.0}
    QuestReward_TextureBlock : texture_block = texture_block{DefaultImage:=Images.EmptyTexture, DefaultDesiredSize := vector2{X:=40.0,Y:=40.0}}
    QuestReward_TextBlock : text_block = text_block{DefaultTextColor:=White, DefaultShadowOffset:=ShadowOffset, DefaultShadowOpacity := 1.0}

    Quest_Background0 : texture_block = texture_block{DefaultImage:=Images.Img_LineBackground, DefaultDesiredSize:=vector2{X:=1.0,Y:=1.0}}
    Quest_Background1 : texture_block = texture_block{DefaultImage:=Images.Img_LineBackground, DefaultDesiredSize:=vector2{X:=1.0,Y:=1.0}}
    Quest_Bolt0 : texture_block = texture_block{DefaultImage:=Images.Dialogue_Bolt, DefaultDesiredSize:=vector2{X:=5.0,Y:=5.0}}
    Quest_Bolt1 : texture_block = texture_block{DefaultImage:=Images.Dialogue_Bolt, DefaultDesiredSize:=vector2{X:=5.0,Y:=5.0}}
    Quest_Bolt2 : texture_block = texture_block{DefaultImage:=Images.Dialogue_Bolt, DefaultDesiredSize:=vector2{X:=5.0,Y:=5.0}}
    Quest_Bolt3 : texture_block = texture_block{DefaultImage:=Images.Dialogue_Bolt, DefaultDesiredSize:=vector2{X:=5.0,Y:=5.0}}

    #Timers
    HexaGameTimer_TextBlock : text_block = text_block{DefaultTextColor:=White, DefaultShadowOffset:=ShadowOffset, DefaultShadowOpacity := 1.0}
    DungeonTimer_TextBlock : text_block = text_block{DefaultTextColor:=White, DefaultShadowOffset:=ShadowOffset, DefaultShadowOpacity := 1.0}
    HexaGameNumTokens_TextBlock : text_block = text_block{DefaultTextColor:=White, DefaultShadowOffset:=ShadowOffset, DefaultShadowOpacity := 1.0}
    DungeonNumTokens_TextBlock : text_block = text_block{DefaultTextColor:=White, DefaultShadowOffset:=ShadowOffset, DefaultShadowOpacity := 1.0}

    #Sapphires
    BackpackSapphires_Ammount : text_block = text_block{DefaultTextColor:=SapphiresTextColor, DefaultShadowOffset:=ShadowOffset, DefaultShadowOpacity := 1.0}
    BackpackSapphires_Icon : texture_block = texture_block{DefaultImage:=Images.Sapphire_Icon, DefaultDesiredSize := vector2{X:=VerySizeSmallIconX,Y:=VerySizeSmallIconY}}
    BackpackSapphires_Background : texture_block = texture_block{DefaultImage:=Images.Resourcers_InventoryUI_Part1, DefaultTint:= SapphiresColor, DefaultDesiredSize:=vector2{X:=205.0,Y:=59.0}}
    BackpackSapphires_Border : texture_block = texture_block{DefaultImage:=Images.InventoryUI_SmallPart, DefaultTint:= SapphiresTextColor, DefaultDesiredSize:=vector2{X:=220.0,Y:=75.0}}
    var SapphiresWarningRunning : logic = false

    #Aged Sapphires
    BackpackAgedSapphires_Ammount : text_block = text_block{DefaultTextColor:=AgedSapphiresTextColor, DefaultShadowOffset:=ShadowOffset, DefaultShadowOpacity := 1.0}
    BackpackAgedSapphires_Icon : texture_block = texture_block{DefaultImage:=Images.AgedSaphire_Icon, DefaultDesiredSize := vector2{X:=VerySizeSmallIconX,Y:=VerySizeSmallIconY}}
    BackpackAgedSapphires_Background : texture_block = texture_block{DefaultImage:=Images.Resourcers_InventoryUI_Part1, DefaultTint:= AgedSapphiresColor, DefaultDesiredSize:=vector2{X:=205.0,Y:=59.0}}
    BackpackAgedSapphires_Border : texture_block = texture_block{DefaultImage:=Images.InventoryUI_SmallPart, DefaultTint:= AgedSapphiresTextColor, DefaultDesiredSize:=vector2{X:=220.0,Y:=75.0}}
    var AgedSapphiresWarningRunning : logic = false

    #Fame
    var IsFameHided : logic = false
    BackpackFame_Amount : text_block = text_block{DefaultTextColor:=FameTextColor, DefaultShadowOffset:=ShadowOffset, DefaultShadowOpacity := 1.0}
    BackpackFame_Icon : texture_block = texture_block{DefaultImage:=Images.Fame_Icon, DefaultDesiredSize := vector2{X:=VerySizeSmallIconX,Y:=VerySizeSmallIconY}}

    #Water
    BackpackWater_Ammount : text_block = text_block{DefaultTextColor:=WaterTextColor, DefaultShadowOffset:=ShadowOffset, DefaultShadowOpacity := 1.0}
    BackpackWater_Icon : texture_block = texture_block{DefaultImage:=Images.Water_Icon, DefaultDesiredSize := vector2{X:=VerySizeSmallIconX,Y:=VerySizeSmallIconY}}
    BackpackWater_Slider : texture_block = texture_block{DefaultTint:=WaterColor,DefaultDesiredSize := vector2{X:=0.0,Y:=0.0},DefaultImage:=Images.Resourcers_InventoryUI_Part1}
    BackpackWater_EmptyPack : texture_block = texture_block{DefaultImage:=Images.Resourcers_InventoryUI_Part0,DefaultDesiredSize:=vector2{X:=219.0,Y:=65.0}}
    BackpackWater_Border : texture_block = texture_block{DefaultImage:=Images.Resourcers_InventoryUI_Part_Water,DefaultDesiredSize:=vector2{X:=265.0,Y:=75.0}}
    
    var WaterWarningRunning : logic = false

    #Flour
    BackpackFlour_Ammount : text_block = text_block{DefaultTextColor:=FlourTextColor, DefaultShadowOffset:=ShadowOffset, DefaultShadowOpacity := 1.0}
    BackpackFlour_Icon : texture_block = texture_block{DefaultImage:=Images.Flour_Icon, DefaultDesiredSize := vector2{X:=VerySizeSmallIconX,Y:=VerySizeSmallIconY}}
    BackpackFlour_Slider : texture_block = texture_block{DefaultTint:=FlourColor,DefaultDesiredSize := vector2{X:=0.0,Y:=0.0},DefaultImage:=Images.Resourcers_InventoryUI_Part1}
    BackpackFlour_EmptyPack : texture_block = texture_block{DefaultImage:=Images.Resourcers_InventoryUI_Part0,DefaultDesiredSize:=vector2{X:=219.0,Y:=68.0}}
    BackpackFlour_Border : texture_block = texture_block{DefaultImage:=Images.Resourcers_InventoryUI_Part_Flour,DefaultDesiredSize:=vector2{X:=265.0,Y:=75.0}}
    var FlourWarningRunning : logic = false

    #Cheese
    BackpackCheese_Ammount : text_block = text_block{DefaultTextColor:=CheeseTextColor, DefaultShadowOffset:=ShadowOffset, DefaultShadowOpacity := 1.0}
    BackpackCheese_Icon : texture_block = texture_block{DefaultImage:=Images.Cheese_Icon, DefaultDesiredSize := vector2{X:=VerySizeSmallIconX,Y:=VerySizeSmallIconY}}
    BackpackCheese_Slider : texture_block = texture_block{DefaultTint:=CheeseColor,DefaultDesiredSize := vector2{X:=0.0,Y:=0.0},DefaultImage:=Images.Resourcers_InventoryUI_Part1}
    BackpackCheese_EmptyPack : texture_block = texture_block{DefaultImage:=Images.Resourcers_InventoryUI_Part0,DefaultDesiredSize:=vector2{X:=219.0,Y:=68.0}}
    BackpackCheese_Border : texture_block = texture_block{DefaultImage:=Images.Resourcers_InventoryUI_Part_Cheese,DefaultDesiredSize:=vector2{X:=265.0,Y:=75.0}}
    var CheeseWarningRunning : logic = false

    #Pimentos
    BackpackPimentos_Ammount : text_block = text_block{DefaultTextColor:=PimentoTextColor, DefaultShadowOffset:=ShadowOffset, DefaultShadowOpacity := 1.0}
    BackpackPimentos_Icon : texture_block = texture_block{DefaultImage:=Images.Pimento_Icon, DefaultDesiredSize := vector2{X:=VerySizeSmallIconX,Y:=VerySizeSmallIconY}}
    BackpackPimentos_Slider : texture_block = texture_block{DefaultTint:=PimentoColor,DefaultDesiredSize := vector2{X:=0.0,Y:=0.0},DefaultImage:=Images.Resourcers_InventoryUI_Part1}
    BackpackPimentos_EmptyPack : texture_block = texture_block{DefaultImage:=Images.Resourcers_InventoryUI_Part0,DefaultDesiredSize:=vector2{X:=219.0,Y:=68.0}}
    BackpackPimentos_Border : texture_block = texture_block{DefaultImage:=Images.Resourcers_InventoryUI_Part_Pimentos,DefaultDesiredSize:=vector2{X:=265.0,Y:=75.0}}
    var PimentosWarningRunning : logic = false

    #MutantMaterial
    BackpackMutantMaterial_Ammount : text_block = text_block{DefaultTextColor:=MutantMaterialTextColor, DefaultShadowOffset:=ShadowOffset, DefaultShadowOpacity := 1.0}
    BackpackMutantMaterial_Icon : texture_block = texture_block{DefaultImage:=Images.Mutant_Material_Icon, DefaultDesiredSize := vector2{X:=VerySizeSmallIconX,Y:=VerySizeSmallIconY}}
    BackpackMutantMaterial_Background : texture_block = texture_block{DefaultImage:=Images.Resourcers_InventoryUI_Part1, DefaultTint:= MutantMaterialColor, DefaultDesiredSize:=vector2{X:=205.0,Y:=59.0}}
    BackpackMutantMaterial_Border : texture_block = texture_block{DefaultImage:=Images.InventoryUI_SmallPart, DefaultTint:= MutantMaterialTextColor, DefaultDesiredSize:=vector2{X:=220.0,Y:=75.0}}
    var MutantMaterialWarningRunning : logic = false

    #Coal
    BackpackCoal_Ammount : text_block = text_block{DefaultTextColor:=CoalTextColor, DefaultShadowOffset:=ShadowOffset, DefaultShadowOpacity := 1.0}
    BackpackCoal_Icon : texture_block = texture_block{DefaultImage:=Images.Coal_Icon, DefaultDesiredSize := vector2{X:=VerySizeSmallIconX,Y:=VerySizeSmallIconY}}
    BackpackCoal_Background : texture_block = texture_block{DefaultImage:=Images.Resourcers_InventoryUI_Part1, DefaultTint:= CoalColor, DefaultDesiredSize:=vector2{X:=205.0,Y:=59.0}}
    BackpackCoal_Border : texture_block = texture_block{DefaultImage:=Images.InventoryUI_SmallPart, DefaultTint:= CoalTextColor, DefaultDesiredSize:=vector2{X:=220.0,Y:=75.0}}
    var CoalWarningRunning : logic = false

    var MaybeCanvas : ?canvas = false
    var UIIsDisplayed : logic = false

    #Close
    CloseButton : button_loud = button_loud{}

    InitUI() : void =
        set AllPetPerk_Text = array{PetPerk0_Text,PetPerk1_Text,PetPerk2_Text}
        set AllPetPerkSlider_Texture = array{PetPerkSlider0_Texture,PetPerkSlider1_Texture,PetPerkSlider2_Texture}
        Canvas := MakeCanvas()
        set MaybeCanvas = option{Canvas}

    SetAgent(InAgent : agent) : void =
        DamageEffect_Texture.SetDesiredSize(vector2{X:=0.0,Y:=0.0})
        set MaybeAgent = option{InAgent}
        set AllCharactersBatch = array{}
        set AllCharactersTimeBatch = array{}
        set AllCharactersSounds = array{}
        set AllCharacterStringBatch = array{}
        set UIIsDisplayed = false
        set IsOrderCompleted_BeingDisplayed = false
        set IsQuestCompleted_BeingDisplayed = false

    DisplayDamageEffect()<suspends> : void =
        ##Print("Displaying damage effect")
        DisplayedVector:vector2=vector2{X:=2100.0,Y:=2100.0}
        CollapsedVector:vector2=vector2{X:=0.0,Y:=0.0}
        DamageEffect_Texture.SetDesiredSize(DisplayedVector)
        Sleep(0.2)
        DamageEffect_Texture.SetDesiredSize(CollapsedVector)
        Sleep(0.2)
        DamageEffect_Texture.SetDesiredSize(DisplayedVector)
        Sleep(0.2)
        DamageEffect_Texture.SetDesiredSize(CollapsedVector)

    <#
    RefreshAllOld(PlayerData : player_data_table, AllLimits : []float, WasOverLimit : []logic) : void =
        Log("Refreshing UI!",LOG_LEVEL_UIs)
        if:
            WaterLimit := AllLimits[WATER]
            FlourLimit := AllLimits[FLOUR]
            CheeseLimit := AllLimits[CHEESE]
            PimentosLimit := AllLimits[PIMENTOS]

            WasOverLimit_Water:=WasOverLimit[WATER]
            WasOverLimit_Flour:=WasOverLimit[FLOUR]
            WasOverLimit_Cheese:=WasOverLimit[CHEESE]
            WasOverLimit_Pimentos:=WasOverLimit[PIMENTOS]
        then:
            RefreshResource(PlayerData.Inventory_Sapphires,-1.0,false,SAPPHIRES)
            RefreshResource(PlayerData.Inventory_AgedSapphires,-1.0,false,AGEDSAPPHIRES)
            RefreshResource(PlayerData.Inventory_Fame,-1.0,false,FAME)
            RefreshResource(PlayerData.Inventory_Water,WaterLimit,WasOverLimit_Water,WATER)
            RefreshResource(PlayerData.Inventory_Flour,FlourLimit,WasOverLimit_Flour,FLOUR)
            RefreshResource(PlayerData.Inventory_Cheese,CheeseLimit,WasOverLimit_Cheese,CHEESE)
            RefreshResource(PlayerData.Inventory_Pimentos,PimentosLimit,WasOverLimit_Pimentos,PIMENTOS)
            RefreshResource(PlayerData.Inventory_MutantMaterial,-1.0,false,MUTANTMATERIAL)
            RefreshResource(PlayerData.Inventory_Coal,-1.0,false,COAL)

            if(PlayerData.HexaGame_Accumulated>=PlayerData.HexaGame_LVAccumulated):
                HexaGameTimer_TextBlock.SetText(BasicMessage("FULL"))
            else:
                HexaTimeTuple:=TransformToMinutes_Seconds_FromFloat(PlayerData.HexaGame_DelayTime)
                HexaGameTimer_TextBlock.SetText(BasicMessage("{HexaTimeTuple(0)}:{HexaTimeTuple(1)}"))

            if(PlayerData.Dungeon_Accumulated>=PlayerData.Dungeon_LVAccumulated):
                DungeonTimer_TextBlock.SetText(BasicMessage("FULL"))
            else:
                DungeonTimeTuple:=TransformToMinutes_Seconds_FromFloat(PlayerData.Dungeon_DelayTime)
                DungeonTimer_TextBlock.SetText(BasicMessage("{DungeonTimeTuple(0)}:{DungeonTimeTuple(1)}"))

            HexaGameNumTokens_TextBlock.SetText(BasicMessage("{PlayerData.HexaGame_Accumulated}"))
            DungeonNumTokens_TextBlock.SetText(BasicMessage("{PlayerData.Dungeon_Accumulated}"))

            if(PlayerData.Order_DelayTime>0.0):
                OrderTimeTuple:=TransformToMinutes_Seconds_FromFloat(PlayerData.Order_DelayTime)
                OrderProgress_TextBlock.SetText(BasicMessage("{OrderTimeTuple(0)}:{OrderTimeTuple(1)}"))

            if(PlayerData.Quest_DelayTime>0.0):
                QuestTimeTuple:=TransformToMinutes_Seconds_FromFloat(PlayerData.Quest_DelayTime)
                QuestProgress_TextBlock.SetText(BasicMessage("{QuestTimeTuple(0)}:{QuestTimeTuple(1)}"))#>

    RefreshPetPerk_Sliders(AllPerkState : []float, AllPetTypes : []int) : void =
        for(Index:=0..AllPetPerkSlider_Texture.Length-1, SliderTexture:=AllPetPerkSlider_Texture[Index],SliderText:=AllPetPerk_Text[Index]):
            if:
                PerkState:=AllPerkState[Index]
                PetType:=AllPetTypes[Index]
            then:
                ##Print("PerkState is {PerkState}")
                if(PerkState>=1.0):
                    SliderTexture.SetTint(Green)
                else:
                    SliderTexture.SetTint(Red)
                if(PetType=ADC):
                    SliderText.SetText(BasicMessage("🏹")) 
                    SliderTexture.SetDesiredSize(vector2{X:=PerkState*PetPerkBoxSizeX,Y:=1.0})
                else if(PetType=DRUID):
                    SliderText.SetText(BasicMessage("💉")) 
                    SliderTexture.SetDesiredSize(vector2{X:=PerkState*PetPerkBoxSizeX,Y:=1.0})
                else if(PetType=TANK):
                    SliderText.SetText(BasicMessage("🛡️")) 
                    SliderTexture.SetDesiredSize(vector2{X:=PerkState*PetPerkBoxSizeX,Y:=1.0})
                else if(PetType=FIGHTER):
                    SliderText.SetText(BasicMessage("🗡️")) 
                    if(PerkState>=1.0):
                        if(ColorBlink=0):
                            set ColorBlink = 1
                            SliderTexture.SetTint(Cyan)
                        else:
                            set ColorBlink = 0
                            SliderTexture.SetTint(Aqua)
                    SliderTexture.SetDesiredSize(vector2{X:=PerkState*PetPerkBoxSizeX,Y:=1.0})
                else if(PetType=HUNTER):
                    SliderText.SetText(BasicMessage("🐾")) 
                    SliderTexture.SetTint(Turquoise)
                    SliderTexture.SetDesiredSize(vector2{X:=PetPerkBoxSizeX,Y:=1.0})
            else:
                SliderText.SetText(BasicMessage("-")) 
                SliderTexture.SetDesiredSize(vector2{X:=0.0,Y:=0.0})
                SliderText.SetTextColor(White)

    RefreshUKResourcers(AllResourcersType : []int, AllResourcersAmounts : []float) : void =
        for(Index:=0..AllResourcersType.Length-1, Type := AllResourcersType[Index], Amount := AllResourcersAmounts[Index]):
            RefreshResource(Amount,5.0,false,Type)

    RefreshResource(Ammount : float, Limit : float, WasOverLimit :logic, Type : int) : void =
        Percentage := Ammount / Limit
        if(WasOverLimit?):
            spawn:
                RunWarning(Ammount,Type)
        else:
            var XSliderValue : float = Percentage * InventoryBoxSizeX
            if(InventoryBoxSizeX<XSliderValue):
                set XSliderValue = InventoryBoxSizeX
            if(Type = SAPPHIRES, not BakeryMachinesManager.IsOnUKTutorial?):
                BackpackSapphires_Ammount.SetText(BasicMessage(GetOptimizedMagnitudeLetter(Ammount))) 
            if(Type = AGEDSAPPHIRES, not BakeryMachinesManager.IsOnUKTutorial?):
                BackpackAgedSapphires_Ammount.SetText(BasicMessage(GetOptimizedMagnitudeLetter(Ammount))) 
            if(Type = FAME, not BakeryMachinesManager.IsOnUKTutorial?):
                BackpackFame_Amount.SetText(BasicMessage(GetOptimizedMagnitudeLetter(Ammount))) 
            if(Type = WATER):
                BackpackWater_Slider.SetDesiredSize(DefaultDesiredSize := vector2{X:=XSliderValue,Y:=ResourceSliderY})
                BackpackWater_Ammount.SetText(BasicMessage(GetOptimizedMagnitudeLetter(Ammount)))  
            if(Type = FLOUR):
                BackpackFlour_Slider.SetDesiredSize(DefaultDesiredSize := vector2{X:=XSliderValue,Y:=ResourceSliderY})
                BackpackFlour_Ammount.SetText(BasicMessage(GetOptimizedMagnitudeLetter(Ammount))) 
            if(Type = CHEESE, not BakeryMachinesManager.IsOnUKTutorial?):
                BackpackCheese_Slider.SetDesiredSize(DefaultDesiredSize := vector2{X:=XSliderValue,Y:=ResourceSliderY})
                BackpackCheese_Ammount.SetText(BasicMessage(GetOptimizedMagnitudeLetter(Ammount))) 
            if(Type = PIMENTOS, not BakeryMachinesManager.IsOnUKTutorial?):
                BackpackPimentos_Slider.SetDesiredSize(DefaultDesiredSize := vector2{X:=XSliderValue,Y:=ResourceSliderY})
                BackpackPimentos_Ammount.SetText(BasicMessage(GetOptimizedMagnitudeLetter(Ammount))) 
            if(Type = MUTANTMATERIAL, not BakeryMachinesManager.IsOnUKTutorial?):
                BackpackMutantMaterial_Ammount.SetText(BasicMessage(GetOptimizedMagnitudeLetter(Ammount))) 
            if(Type = COAL, not BakeryMachinesManager.IsOnUKTutorial?):
                BackpackCoal_Ammount.SetText(BasicMessage(GetOptimizedMagnitudeLetter(Ammount))) 
 
    RunWarning(Ammount : float, Type : int)<suspends> : void =
        if(Type = WATER):
            if(not WaterWarningRunning?):
                set WaterWarningRunning = true
                var Loop : int = 0
                loop:
                    BackpackWater_Slider.SetDesiredSize(DefaultDesiredSize := vector2{X:=InventoryBoxSizeX})
                    BackpackWater_Ammount.SetText(BasicMessage("FULL!!!!"))
                    BackpackWater_Ammount.SetShadowOffset(option{vector2{X:=3.0,Y:=3.0}})
                    Sleep(0.25)
                    BackpackWater_Ammount.SetShadowOffset(ShadowOffset)
                    Sleep(0.25)
                    set Loop = Loop + 1
                    if(Loop>=15):
                        BackpackWater_Ammount.SetText(BasicMessage(GetOptimizedMagnitudeLetter(Ammount)))   
                        BackpackWater_Ammount.SetShadowOffset(ShadowOffset)
                        set WaterWarningRunning = false
                        break
        
        if(Type = SAPPHIRES):
            if(not SapphiresWarningRunning?):
                set SapphiresWarningRunning = true
                var Loop : int = 0
                loop:
                    Sleep(0.25)
                    BackpackSapphires_Ammount.SetShadowOffset(ShadowOffset)
                    Sleep(0.25)
                    set Loop = Loop + 1
                    if(Loop>=15):
                        BackpackSapphires_Ammount.SetText(BasicMessage(GetOptimizedMagnitudeLetter(Ammount)))   
                        BackpackSapphires_Ammount.SetShadowOffset(ShadowOffset)
                        set SapphiresWarningRunning = false
                        break
        
        if(Type = AGEDSAPPHIRES):
            if(not AgedSapphiresWarningRunning?):
                set AgedSapphiresWarningRunning = true
                var Loop : int = 0
                loop:
                    BackpackAgedSapphires_Ammount.SetText(BasicMessage("FULL!!!!"))
                    BackpackAgedSapphires_Ammount.SetShadowOffset(option{vector2{X:=3.0,Y:=3.0}})
                    Sleep(0.25)
                    BackpackAgedSapphires_Ammount.SetShadowOffset(ShadowOffset)
                    Sleep(0.25)
                    set Loop = Loop + 1
                    if(Loop>=15):
                        BackpackAgedSapphires_Ammount.SetText(BasicMessage(GetOptimizedMagnitudeLetter(Ammount)))   
                        BackpackAgedSapphires_Ammount.SetShadowOffset(ShadowOffset)
                        set AgedSapphiresWarningRunning = false
                        break
        
        if(Type = FLOUR):
            if(not FlourWarningRunning?):
                set FlourWarningRunning = true
                var Loop : int = 0
                loop:
                    BackpackFlour_Slider.SetDesiredSize(DefaultDesiredSize := vector2{X:=InventoryBoxSizeX,Y:=0.0})
                    BackpackFlour_Ammount.SetText(BasicMessage("FULL!!!!"))
                    BackpackFlour_Ammount.SetShadowOffset(option{vector2{X:=3.0,Y:=3.0}})
                    Sleep(0.25)
                    BackpackFlour_Ammount.SetShadowOffset(ShadowOffset)
                    Sleep(0.25)
                    set Loop = Loop + 1
                    if(Loop>=15):
                        BackpackFlour_Ammount.SetText(BasicMessage(GetOptimizedMagnitudeLetter(Ammount)))   
                        BackpackFlour_Ammount.SetShadowOffset(ShadowOffset)
                        set FlourWarningRunning = false
                        break
        
        if(Type = CHEESE):
            if(not CheeseWarningRunning?):
                set CheeseWarningRunning = true
                var Loop : int = 0
                loop:
                    BackpackCheese_Slider.SetDesiredSize(DefaultDesiredSize := vector2{X:=InventoryBoxSizeX,Y:=0.0})
                    BackpackCheese_Ammount.SetText(BasicMessage("FULL!!!!"))
                    BackpackCheese_Ammount.SetShadowOffset(option{vector2{X:=3.0,Y:=3.0}})
                    Sleep(0.25)
                    BackpackCheese_Ammount.SetShadowOffset(ShadowOffset)
                    Sleep(0.25)
                    set Loop = Loop + 1
                    if(Loop>=15):
                        BackpackCheese_Ammount.SetText(BasicMessage(GetOptimizedMagnitudeLetter(Ammount)))   
                        BackpackCheese_Ammount.SetShadowOffset(ShadowOffset)
                        set CheeseWarningRunning = false
                        break
        
        if(Type = PIMENTOS):
            if(not PimentosWarningRunning?):
                set PimentosWarningRunning = true
                var Loop : int = 0
                loop:
                    BackpackPimentos_Slider.SetDesiredSize(DefaultDesiredSize := vector2{X:=InventoryBoxSizeX,Y:=0.0})
                    BackpackPimentos_Ammount.SetText(BasicMessage("FULL!!!!"))
                    BackpackPimentos_Ammount.SetShadowOffset(option{vector2{X:=3.0,Y:=3.0}})
                    Sleep(0.25)
                    BackpackPimentos_Ammount.SetShadowOffset(ShadowOffset)
                    Sleep(0.25)
                    set Loop = Loop + 1
                    if(Loop>=15):
                        BackpackPimentos_Ammount.SetText(BasicMessage(GetOptimizedMagnitudeLetter(Ammount)))   
                        BackpackPimentos_Ammount.SetShadowOffset(ShadowOffset)
                        set PimentosWarningRunning = false
                        break
        
        if(Type = MUTANTMATERIAL):
            if(not MutantMaterialWarningRunning?):
                set MutantMaterialWarningRunning = true
                var Loop : int = 0
                loop:
                    BackpackMutantMaterial_Ammount.SetText(BasicMessage("FULL!!!!"))
                    BackpackMutantMaterial_Ammount.SetShadowOffset(option{vector2{X:=3.0,Y:=3.0}})
                    Sleep(0.25)
                    BackpackMutantMaterial_Ammount.SetShadowOffset(ShadowOffset)
                    Sleep(0.25)
                    set Loop = Loop + 1
                    if(Loop>=15):
                        BackpackMutantMaterial_Ammount.SetText(BasicMessage(GetOptimizedMagnitudeLetter(Ammount)))   
                        BackpackMutantMaterial_Ammount.SetShadowOffset(ShadowOffset)
                        set MutantMaterialWarningRunning = false
                        break
        
        if(Type = COAL):
            if(not CoalWarningRunning?):
                set CoalWarningRunning = true
                var Loop : int = 0
                loop:
                    BackpackCoal_Ammount.SetText(BasicMessage("FULL!!!!"))
                    BackpackCoal_Ammount.SetShadowOffset(option{vector2{X:=3.0,Y:=3.0}})
                    Sleep(0.25)
                    BackpackCoal_Ammount.SetShadowOffset(ShadowOffset)
                    Sleep(0.25)
                    set Loop = Loop + 1
                    if(Loop>=15):
                        BackpackCoal_Ammount.SetText(BasicMessage(GetOptimizedMagnitudeLetter(Ammount)))   
                        BackpackCoal_Ammount.SetShadowOffset(ShadowOffset)
                        set CoalWarningRunning = false
                        break

    DisplayCharacterMessage(Character : int, InString : ?string, InMessage : ?message, Sound : int, ?DisplayTime : float = 5.0, ?Priority:int=MIN_PRIORITY) : void =

        #Log("Trying to display character message: {Phase0State}",LOG_LEVEL_UK)
        #If it is top priority, clean all messages

        if(AllCharacterStringBatch.Length=0, AllCharacterMaybeMessagesBatch.Length=0):
            set AllCharacterStringBatch = AllCharacterStringBatch + array{InString}
            set AllCharacterMaybeMessagesBatch = AllCharacterMaybeMessagesBatch + array{InMessage}
            set AllCharactersBatch = AllCharactersBatch + array{Character}
            set AllCharactersTimeBatch = AllCharactersTimeBatch + array{DisplayTime}
            set AllCharactersSounds = AllCharactersSounds + array{Sound}
            if(not CharacterMessageIsRunning?):
                set CharacterMessageIsRunning = true
                spawn:
                    CharacterMessageCue()
        else:
            if(AllCharacterStringBatch.Length>0, AllCharacterMaybeMessagesBatch.Length>0, Priority=MAX_PRIORITY):
                set AllCharacterMaybeMessagesBatch = array{InMessage}
                set AllCharacterStringBatch = array{InString}
                set AllCharactersBatch = array{Character}
                set AllCharactersTimeBatch = array{DisplayTime}
                set AllCharactersSounds = array{Sound}

            #If it is low priority, dont add
            if(AllCharacterStringBatch.Length>0, AllCharacterMaybeMessagesBatch.Length>0, Priority=MIN_PRIORITY):
                Log("Message not added because was low priority",LOG_LEVEL_DEBUG7)
            else:
                set AllCharacterStringBatch = AllCharacterStringBatch + array{InString}
                set AllCharacterMaybeMessagesBatch = AllCharacterMaybeMessagesBatch + array{InMessage}
                set AllCharactersBatch = AllCharactersBatch + array{Character}
                set AllCharactersTimeBatch = AllCharactersTimeBatch + array{DisplayTime}
                set AllCharactersSounds = AllCharactersSounds + array{Sound}
                if(not CharacterMessageIsRunning?):
                    set CharacterMessageIsRunning = true
                    spawn:
                        CharacterMessageCue()
            if(AllCharacterStringBatch.Length>3, AllCharacterMaybeMessagesBatch.Length>3):
                if:
                    LastMessage := AllCharacterMaybeMessagesBatch[AllCharacterMaybeMessagesBatch.Length-1]
                    LastString := AllCharacterStringBatch[AllCharacterStringBatch.Length-1]
                    LastCharacter := AllCharactersBatch[AllCharactersBatch.Length-1]
                    LastTime := AllCharactersTimeBatch[AllCharactersTimeBatch.Length-1]
                    LastSound := AllCharactersSounds[AllCharactersTimeBatch.Length-1]
                    set AllCharacterMaybeMessagesBatch = array{LastMessage}
                    set AllCharacterStringBatch = array{LastString}
                    set AllCharactersBatch = array{LastCharacter}
                    set AllCharactersTimeBatch = array{LastTime}
                    set AllCharactersSounds = array{Sound}

                PrintWarning("Max CharacterMessageCue exceded")

    CleanCharacterMessages() : void =
        set AllCharacterMaybeMessagesBatch = array{}
        set AllCharacterStringBatch = array{}
        set AllCharactersBatch = array{}
        set AllCharactersTimeBatch = array{}
        set AllCharactersSounds = array{}


    CharacterMessageCue<private>()<suspends> : void =
        loop:
            if:
                AgentOwner:=BakeryMachinesManager.MaybeAgentOwner?
                AllCharacterStringBatch.Length>0
                AllCharacterMaybeMessagesBatch.Length>0
                MaybeNextString := AllCharacterStringBatch[0]
                MaybeNextMessage := AllCharacterMaybeMessagesBatch[0]
                NextCharacter := AllCharactersBatch[0]
                NextDelay := AllCharactersTimeBatch[0]
                NextSound := AllCharactersSounds[0]
                NewMessages := AllCharacterMaybeMessagesBatch.RemoveElement[0]
                NewStrings := AllCharacterStringBatch.RemoveElement[0]
                NewCharactersArray := AllCharactersBatch.RemoveElement[0]
                NewTimedArray := AllCharactersTimeBatch.RemoveElement[0]
                NewSoundsArray := AllCharactersSounds.RemoveElement[0]
            then:
                set AllCharacterMaybeMessagesBatch = NewMessages
                set AllCharacterStringBatch = NewStrings
                set AllCharactersBatch = NewCharactersArray
                set AllCharactersTimeBatch = NewTimedArray
                set AllCharactersSounds = NewSoundsArray
                #Play Sound
                if(NextCharacter <> I_CHARACTER or NextSound=-1):
                    if(NextCharacter=W_CHARACTER):
                        BakeryMachinesManager.GameManager.VFXManager.WCharacter_AudioPlayer.Play(AgentOwner)
                    else if(NextCharacter=G_CHARACTER):
                        BakeryMachinesManager.GameManager.VFXManager.GCharacter_AudioPlayer.Play(AgentOwner)
                    else if(NextCharacter=J_CHARACTER):
                        BakeryMachinesManager.GameManager.VFXManager.JCharacter_AudioPlayer.Play(AgentOwner)
                    else if(NextCharacter=H_CHARACTER):
                        BakeryMachinesManager.GameManager.VFXManager.HCharacter_AudioPlayer.Play(AgentOwner)
                else:
                    spawn:
                        BakeryMachinesManager.GameManager.VFXManager.PlaySound(NextSound,0.0,option{AgentOwner},false)
                #Set Character Texture
                if(NextCharacter=W_CHARACTER):
                    CharacterImage_TextureBlock.SetImage(Images.Jhonny_Icon)
                    CharacterImage_TextureBlockLoc.SetImage(Images.Jhonny_Icon)
                else if(NextCharacter=G_CHARACTER):
                    CharacterImage_TextureBlock.SetImage(Images.Mike_Icon)
                    CharacterImage_TextureBlockLoc.SetImage(Images.Mike_Icon)
                else if(NextCharacter=J_CHARACTER):
                    CharacterImage_TextureBlock.SetImage(Images.Steve_Icon)
                    CharacterImage_TextureBlockLoc.SetImage(Images.Steve_Icon)
                else if(NextCharacter=H_CHARACTER):
                    CharacterImage_TextureBlock.SetImage(Images.Frankie_Icon)
                    CharacterImage_TextureBlockLoc.SetImage(Images.Frankie_Icon)
                else if(NextCharacter=I_CHARACTER):
                    CharacterImage_TextureBlock.SetImage(Images.Info_Icon)
                    CharacterImage_TextureBlockLoc.SetImage(Images.Info_Icon)

                #Adapting String
                if(ValidMessage := MaybeNextMessage?):
                    CharacterMessage_TextBlock.SetText(ValidMessage)
                    CharacterMessage_TextBlockLoc.SetText(ValidMessage)
                    RunCharacterMessage(NextDelay,true)
                else if(NextString := MaybeNextString?):
                    FinalString:=AdaptStringToLine(NextString,30)
                    CharacterMessage_TextBlock.SetText(BasicMessage(FinalString))
                    CharacterMessage_TextBlockLoc.SetText(BasicMessage(FinalString))
                    RunCharacterMessage(NextDelay,false)
            else:
                break
            
            Sleep(0.1)
            if(AllCharacterStringBatch.Length>3, AllCharacterMaybeMessagesBatch.Length>3, LastMessage := AllCharacterMaybeMessagesBatch[AllCharacterMaybeMessagesBatch.Length-1], LastString := AllCharacterStringBatch[AllCharacterStringBatch.Length-1], set AllCharacterStringBatch = array{LastString}, set AllCharacterMaybeMessagesBatch = array{LastMessage}):
                PrintWarning("Max CharacterMessageCue exceded")
        set CharacterMessageIsRunning = false

    DisplayOrder(PlayerData : player_data_table, Type : int, CurrentAmount : float, ObjetiveAmount : float, RewardType : int, Reward : float) : void =
        if(Agent := MaybeAgent?):
            MaybeHexaManager:=BakeryMachinesManager.GameManager.GetAgentHexaManager(Agent)
            if:
                HexaManager:=MaybeHexaManager?
                HexMiniGameUI:=HexaManager.MaybeHexMiniGameUI?
                HexSelectorUI:=HexaManager.MaybeHexMiniGameSelectorUI?
                not HexMiniGameUI.UIIsDisplayed?
                not HexSelectorUI.UIIsDisplayed?
                not HexaManager.UIIsChallengeCanvas?
                not HexaManager.UIIsNavigationCanvas?
            then: 
                ##Print("Displaying order")
                var TitleString : string = ""
                OrderProgress_TextBlock.SetTextColor(CurrentTextColor)
                OrderProgress_TextBlock.SetText(BasicMessage("{GetMagnitudeLetter(CurrentAmount, "", "")}/{GetMagnitudeLetter(ObjetiveAmount, "", "")}"))
                OrderProduct_TextureBlock.SetDesiredSize(vector2{X:=40.0,Y:=40.0})
                #OrderDialogue_Slider.SetDesiredSize(vector2{X:=OrderMessage_SliderDisplayedAmount,Y:=0.0})
                #Current objetive
                
                if(Type=B_TYPE):
                    set TitleString = "Bake Bread"
                    OrderProduct_TextureBlock.SetDesiredSize(vector2{X:=40.0,Y:=40.0})
                    OrderProduct_TextureBlock.SetImage(Images.Bread_Icon)
                else if(Type=CB_TYPE):
                    set TitleString = "Bake Cheese Bread"
                    OrderProduct_TextureBlock.SetDesiredSize(vector2{X:=40.0,Y:=40.0})
                    OrderProduct_TextureBlock.SetImage(Images.CheeseBread_Icon)
                else if(Type=PCS_TYPE):
                    set TitleString = "Bake Sandwiches"
                    OrderProduct_TextureBlock.SetDesiredSize(vector2{X:=40.0,Y:=40.0})
                    OrderProduct_TextureBlock.SetImage(Images.PimentoCheeseSandwich_Icon)
                else:
                    set TitleString = "Next order..."
                    RefreshTimers(PlayerData)
                    OrderProduct_TextureBlock.SetDesiredSize(vector2{X:=0.0,Y:=0.0})
                    OrderProduct_TextureBlock.SetImage(Images.EmptyTexture)

                OrderRewardDescrip_TextBlock.SetTextColor(White)
                OrderRewardDescrip_TextBlock.SetText(BasicMessage("Reward:"))

                #Rewards
                if(RewardType=COAL):
                    OrderReward_TextureBlock.SetImage(Images.Coal_Icon)
                    OrderReward_TextureBlock.SetDesiredSize(vector2{X:=40.0,Y:=40.0})
                    OrderReward_TextBlock.SetTextColor(CoalTextColor)
                    OrderReward_TextBlock.SetText(BasicMessage(GetMagnitudeLetter(Reward,"","")))
                else if(RewardType=SAPPHIRES): 
                    OrderReward_TextureBlock.SetImage(Images.Sapphire_Icon)
                    OrderReward_TextureBlock.SetDesiredSize(vector2{X:=40.0,Y:=40.0})
                    OrderReward_TextBlock.SetTextColor(SapphiresTextColor)
                    OrderReward_TextBlock.SetText(BasicMessage(GetMagnitudeLetter(Reward,"","")))
                else if(RewardType=MUTANTMATERIAL): 
                    OrderReward_TextureBlock.SetImage(Images.Mutant_Material_Icon)
                    OrderReward_TextureBlock.SetDesiredSize(vector2{X:=40.0,Y:=40.0})
                    OrderReward_TextBlock.SetTextColor(MutantMaterialTextColor)
                    OrderReward_TextBlock.SetText(BasicMessage(GetMagnitudeLetter(Reward,"","")))
                else: 
                    OrderReward_TextureBlock.SetImage(Images.EmptyTexture)
                    OrderReward_TextureBlock.SetDesiredSize(vector2{X:=0.0,Y:=0.0})
                    OrderReward_TextBlock.SetTextColor(White)
                    OrderReward_TextBlock.SetText(BasicMessage("-"))

                
                    <#
                    set TitleString ="Complete the tutorial"
                    OrderProgress_TextBlock.SetText(BasicMessage("To unlock features"))
                    OrderReward_TextBlock.SetText(BasicMessage(""))
                    OrderRewardDescrip_TextBlock.SetText(BasicMessage("Good luck!"))
                    OrderRewardDescrip_TextBlock.SetTextColor(MaxedColor)#>

                TitleString_Final:=AdaptStringToLine(TitleString,20)
                OrderTitle_TextBlock.SetText(BasicMessage(TitleString))
                if(PlayerData.NumQuestsCompleted <= TutorialOrder_Index):
                    Hide_OrderPanel()


    Hide_QuestPanel() : void =
        QuestTitle_TextBlock.SetText(BasicMessage(""))
        QuestIcon_TextureBlock.SetDesiredSize(vector2{})

        QuestDescription_TextBlock.SetText(BasicMessage(""))
        QuestRewardDescrip_TextBlock.SetText(BasicMessage(""))
        QuestReward_TextBlock.SetText(BasicMessage(""))
        OrderRewardDescrip_TextBlock.SetText(BasicMessage(""))

        QuestReward_TextureBlock.SetDesiredSize(vector2{})

        QuestProgress_TextBlock.SetText(BasicMessage(""))

        Quest_Bolt0.SetDesiredSize(vector2{})
        Quest_Bolt1.SetDesiredSize(vector2{})
        Quest_Bolt2.SetDesiredSize(vector2{})
        Quest_Bolt3.SetDesiredSize(vector2{})
        Quest_Background0.SetImage(Images.EmptyTexture)
        Quest_Background1.SetImage(Images.EmptyTexture)

    Show_QuestPanel() : void =
        Quest_Bolt0.SetDesiredSize(SmallBoltDesiredSize)
        Quest_Bolt1.SetDesiredSize(SmallBoltDesiredSize)
        Quest_Bolt2.SetDesiredSize(SmallBoltDesiredSize)
        Quest_Bolt3.SetDesiredSize(SmallBoltDesiredSize)
        Quest_Background0.SetImage(Images.Img_LineBackground)
        Quest_Background1.SetImage(Images.Img_LineBackground)

    SetUKQuest(MessageDescription : message) : void =

        BackpackWater_EmptyPack.SetDesiredSize(Limited_EmptyPack)
        BackpackWater_Icon.SetDesiredSize(VerySizeSmallIcon)
        BackpackWater_Border.SetDesiredSize(Limited_Border)

        BackpackFlour_EmptyPack.SetDesiredSize(Limited_EmptyPack)
        BackpackFlour_Icon.SetDesiredSize(VerySizeSmallIcon)
        BackpackFlour_Border.SetDesiredSize(Limited_Border)

        if(UKTutorial:=BakeryMachinesManager.MaybeUKTutorial?):
            if(BakeryMachinesManager.PlayerDataCache.Language=ENGLISH):
                QuestTitle_TextBlock.SetText(BasicMessage(otherstrings_en_module.Title))
                QuestReward_TextBlock.SetText(BasicMessage(otherstrings_en_module.GoodLuck))
            else if(BakeryMachinesManager.PlayerDataCache.Language=SPANISH):
                QuestTitle_TextBlock.SetText(BasicMessage(otherstrings_es_module.Title))
                QuestReward_TextBlock.SetText(BasicMessage(otherstrings_es_module.GoodLuck))
            else if(BakeryMachinesManager.PlayerDataCache.Language=FRENCH):
                QuestTitle_TextBlock.SetText(BasicMessage(otherstrings_fr_module.Title))
                QuestReward_TextBlock.SetText(BasicMessage(otherstrings_fr_module.GoodLuck))
            else if(BakeryMachinesManager.PlayerDataCache.Language=PORTUGUESE):
                QuestTitle_TextBlock.SetText(BasicMessage(otherstrings_pt_module.Title))
                QuestReward_TextBlock.SetText(BasicMessage(otherstrings_pt_module.GoodLuck))
        
            QuestIcon_TextureBlock.SetImage(UKTutorial.CurrentTutorial_Icon)
            QuestRewardDescrip_TextBlock.SetText(BasicMessage(""))
            QuestReward_TextureBlock.SetDesiredSize(vector2{})
            QuestDescription_TextBlock.SetText(MessageDescription)
        
            # UK Tutorial Quest Progress
            ProgressAmount := GetOptimizedMagnitudeLetter(UKTutorial.CurrentTutorialProgress_Amount)
            ObjetiveAmount := GetOptimizedMagnitudeLetter(UKTutorial.CurrentTutorialObjetive_Amount)
        
            QuestIcon_TextureBlock.SetDesiredSize(vector2{X:=40.0,Y:=40.0})
            QuestIcon_TextureBlock.SetImage(UKTutorial.CurrentTutorial_Icon)
            QuestProgress_TextBlock.SetText(BasicMessage("{ProgressAmount}/{ObjetiveAmount}"))
        

    DisplayQuest(PlayerData : player_data_table) : void =
        if(Agent := MaybeAgent?):
            MaybeHexaManager:=BakeryMachinesManager.GameManager.GetAgentHexaManager(Agent)
            if:
                HexaManager:=MaybeHexaManager?
                HexMiniGameUI:=HexaManager.MaybeHexMiniGameUI?
                HexSelectorUI:=HexaManager.MaybeHexMiniGameSelectorUI?
                not HexMiniGameUI.UIIsDisplayed?
                not HexSelectorUI.UIIsDisplayed?
                not HexaManager.UIIsChallengeCanvas?
                not HexaManager.UIIsNavigationCanvas?
            then: 
                if(not BakeryMachinesManager.IsOnUKTutorial?):
                    ##Print("Displaying quest")
                    var TitleString : string = ""
                    var DescriptionString : string = ""
                    var ProgressString : string = ""

                    QuestInfoTuple:=BakeryMachinesManager.GameManager.QuestsAndOrdersDatabase.GetCurrentQuest(PlayerData.NumQuestsCompleted,PlayerData)
                    QuestRewardDescrip_TextBlock.SetText(BasicMessage("Reward:"))

                    ##Print("Displaying quest {PlayerData.NumQuestCompleted}")

                    if(PlayerData.NumQuestsCompleted<=Quest0_Index or PlayerData.Quest_DelayTime=0.0):
                        set TitleString = QuestInfoTuple(0)

                        QuestIcon_TextureBlock.SetDesiredSize(vector2{X:=40.0,Y:=40.0})
                        QuestIcon_TextureBlock.SetImage(QuestInfoTuple(1))

                        QuestProgress_TextBlock.SetTextColor(UpgradeTextColor)

                        set DescriptionString = QuestInfoTuple(2)
                        QuestDescription_TextBlock.SetTextColor(CurrentTextColor)

                        QuestReward_TextBlock.SetText(BasicMessage(GetOptimizedMagnitudeLetter(QuestInfoTuple(6))))
                    
                        if(SAPPHIRES=QuestInfoTuple(5)):
                            QuestReward_TextureBlock.SetImage(SapphireIcon)
                            QuestReward_TextBlock.SetTextColor(SapphiresTextColor)
                        else if(AGEDSAPPHIRES=QuestInfoTuple(5)):
                            QuestReward_TextureBlock.SetImage(AgedSapphireIcon)
                            QuestReward_TextBlock.SetTextColor(AgedSapphiresTextColor)
                        else if(COAL=QuestInfoTuple(5)):
                            QuestReward_TextureBlock.SetImage(CoalIcon)
                            QuestReward_TextBlock.SetTextColor(CoalTextColor)
                        else if(MUTANTMATERIAL=QuestInfoTuple(5)):
                            QuestReward_TextureBlock.SetImage(MutantIcon)
                            QuestReward_TextBlock.SetTextColor(MutantMaterialTextColor)
                        else if(WATER=QuestInfoTuple(5)):
                            QuestReward_TextureBlock.SetImage(WaterTexture)
                            QuestReward_TextBlock.SetTextColor(WaterTextColor)
                        else if(FLOUR=QuestInfoTuple(5)):
                            QuestReward_TextureBlock.SetImage(FlourTexture)
                            QuestReward_TextBlock.SetTextColor(FlourTextColor)
                        else:
                            QuestReward_TextBlock.SetTextColor(White)

                        QuestReward_TextureBlock.SetDesiredSize(VerySizeSmallIcon)

                        if:
                            B_TYPE=QuestInfoTuple(3) or 
                            CB_TYPE=QuestInfoTuple(3) or 
                            PCS_TYPE=QuestInfoTuple(3) or
                            CHEESE=QuestInfoTuple(3) or
                            PIMENTOS=QuestInfoTuple(3) or
                            MUTANTMATERIAL=QuestInfoTuple(3)
                        then:
                            ProgressAmount := GetOptimizedMagnitudeLetter(PlayerData.QuestAssigned_ProgressAmount)
                            ObjetiveAmount := GetOptimizedMagnitudeLetter(QuestInfoTuple(4))

                            DescriptionString_Final:=AdaptStringToLineV2(DescriptionString,22)
                            QuestDescription_TextBlock.SetText(BasicMessage(DescriptionString_Final))
                            QuestProgress_TextBlock.SetText(BasicMessage("{ProgressAmount}/{ObjetiveAmount}"))
                        else if (IntProgressAmount := Floor[PlayerData.QuestAssigned_ProgressAmount], IntObjetiveAmount:=Floor[QuestInfoTuple(4)]):

                            DescriptionString_Final:=AdaptStringToLineV2(DescriptionString,22)
                            QuestDescription_TextBlock.SetText(BasicMessage(DescriptionString_Final))
                            QuestProgress_TextBlock.SetText(BasicMessage("{IntProgressAmount}/{IntObjetiveAmount}"))


                    else:
                        set TitleString = "Next quest..."
                        QuestIcon_TextureBlock.SetDesiredSize(vector2{})

                        HexaTimeTuple:=TransformToMinutes_Seconds_FromFloat(PlayerData.Quest_DelayTime)
                        QuestDescription_TextBlock.SetText(BasicMessage("{HexaTimeTuple(0)}:{HexaTimeTuple(1)}"))
                        QuestDescription_TextBlock.SetTextColor(Red)

                        QuestReward_TextureBlock.SetDesiredSize(vector2{})
                        QuestReward_TextBlock.SetText(BasicMessage("-"))
                        QuestProgress_TextBlock.SetText(BasicMessage("-"))

                
                    TitleString_Final:=AdaptStringToLine(TitleString,20)
                    QuestTitle_TextBlock.SetText(BasicMessage(TitleString_Final))
                else if(UKTutorial:=BakeryMachinesManager.MaybeUKTutorial?):



                    #UK Tutorial Quest
                    # UK Tutorial Quest
                    if(BakeryMachinesManager.PlayerDataCache.Language=ENGLISH):
                        QuestTitle_TextBlock.SetText(BasicMessage(otherstrings_en_module.Title))
                        QuestReward_TextBlock.SetText(BasicMessage(otherstrings_en_module.GoodLuck))
                    else if(BakeryMachinesManager.PlayerDataCache.Language=SPANISH):
                        QuestTitle_TextBlock.SetText(BasicMessage(otherstrings_es_module.Title))
                        QuestReward_TextBlock.SetText(BasicMessage(otherstrings_es_module.GoodLuck))
                    else if(BakeryMachinesManager.PlayerDataCache.Language=FRENCH):
                        QuestTitle_TextBlock.SetText(BasicMessage(otherstrings_fr_module.Title))
                        QuestReward_TextBlock.SetText(BasicMessage(otherstrings_fr_module.GoodLuck))
                    else if(BakeryMachinesManager.PlayerDataCache.Language=PORTUGUESE):
                        QuestTitle_TextBlock.SetText(BasicMessage(otherstrings_pt_module.Title))
                        QuestReward_TextBlock.SetText(BasicMessage(otherstrings_pt_module.GoodLuck))

                    ProgressAmount := GetOptimizedMagnitudeLetter(UKTutorial.CurrentTutorialProgress_Amount)
                    ObjetiveAmount := GetOptimizedMagnitudeLetter(UKTutorial.CurrentTutorialObjetive_Amount)

                    QuestIcon_TextureBlock.SetDesiredSize(vector2{X:=40.0,Y:=40.0})
                    QuestIcon_TextureBlock.SetImage(UKTutorial.CurrentTutorial_Icon)
                    QuestRewardDescrip_TextBlock.SetText(BasicMessage(""))
                    QuestProgress_TextBlock.SetText(BasicMessage("{ProgressAmount}/{ObjetiveAmount}"))
                    QuestReward_TextureBlock.SetDesiredSize(vector2{})



    Hide_All() : void =
        Hide_OrderPanel()
        Hide_QuestPanel()
        Hide_ResourcersPanel()
        Hide_Fame()


    Hide_OrderPanel() : void =
        OrderTitle_TextBlock.SetText(BasicMessage(""))
        OrderProgress_TextBlock.SetText(BasicMessage(""))
        OrderReward_TextBlock.SetText(BasicMessage(""))
        OrderRewardDescrip_TextBlock.SetText(BasicMessage(""))

        OrderReward_TextureBlock.SetDesiredSize(vector2{})
        OrderProduct_TextureBlock.SetDesiredSize(vector2{})

        Order_Bolt0.SetDesiredSize(vector2{})
        Order_Bolt1.SetDesiredSize(vector2{})
        Order_Bolt2.SetDesiredSize(vector2{})
        Order_Bolt3.SetDesiredSize(vector2{})
        Order_Background0.SetImage(Images.EmptyTexture)
        Order_Background1.SetImage(Images.EmptyTexture)

    Show_OrderPanel() : void =
        Order_Bolt0.SetDesiredSize(SmallBoltDesiredSize)
        Order_Bolt1.SetDesiredSize(SmallBoltDesiredSize)
        Order_Bolt2.SetDesiredSize(SmallBoltDesiredSize)
        Order_Bolt3.SetDesiredSize(SmallBoltDesiredSize)
        Order_Background0.SetImage(Images.Img_LineBackground)
        Order_Background1.SetImage(Images.Img_LineBackground)

    Hide_ResourcersPanel() : void =
        BackpackSapphires_Ammount.SetText(BasicMessage(""))
        BackpackSapphires_Background.SetDesiredSize(HideSize)
        BackpackSapphires_Border.SetDesiredSize(HideSize)
        BackpackSapphires_Icon.SetDesiredSize(HideSize)

        BackpackCoal_Ammount.SetText(BasicMessage(""))
        BackpackCoal_Background.SetDesiredSize(HideSize)
        BackpackCoal_Border.SetDesiredSize(HideSize)
        BackpackCoal_Icon.SetDesiredSize(HideSize)
        
        BackpackAgedSapphires_Ammount.SetText(BasicMessage(""))
        BackpackAgedSapphires_Background.SetDesiredSize(HideSize)
        BackpackAgedSapphires_Border.SetDesiredSize(HideSize)
        BackpackAgedSapphires_Icon.SetDesiredSize(HideSize)

        #Water
        BackpackWater_Ammount.SetText(BasicMessage(""))
        BackpackWater_EmptyPack.SetDesiredSize(HideSize)
        BackpackWater_Slider.SetDesiredSize(HideSize)
        BackpackWater_Icon.SetDesiredSize(HideSize)
        BackpackWater_Border.SetDesiredSize(HideSize)

        #Flour
        BackpackFlour_Ammount.SetText(BasicMessage(""))
        BackpackFlour_EmptyPack.SetDesiredSize(HideSize)
        BackpackFlour_Slider.SetDesiredSize(HideSize)
        BackpackFlour_Icon.SetDesiredSize(HideSize)
        BackpackFlour_Border.SetDesiredSize(HideSize)

        #Cheese
        BackpackCheese_Ammount.SetText(BasicMessage(""))
        BackpackCheese_EmptyPack.SetDesiredSize(HideSize)
        BackpackCheese_Slider.SetDesiredSize(HideSize)
        BackpackCheese_Icon.SetDesiredSize(HideSize)
        BackpackCheese_Border.SetDesiredSize(HideSize)
       
        #Pimentos
        BackpackPimentos_Ammount.SetText(BasicMessage(""))
        BackpackPimentos_EmptyPack.SetDesiredSize(HideSize)
        BackpackPimentos_Slider.SetDesiredSize(HideSize)
        BackpackPimentos_Icon.SetDesiredSize(HideSize)
        BackpackPimentos_Border.SetDesiredSize(HideSize)
       
        #Mutant Material
        BackpackMutantMaterial_Ammount.SetText(BasicMessage(""))
        BackpackMutantMaterial_Background.SetDesiredSize(HideSize)
        BackpackMutantMaterial_Border.SetDesiredSize(HideSize)
        BackpackMutantMaterial_Icon.SetDesiredSize(HideSize)

    RefreshTimers(PlayerData : player_data_table) : void =
        #HEXA GAMES AND DUNGEONS
            #HexaGames
        HexaGameNumTokens_TextBlock.SetText(BasicMessage("{PlayerData.HexaGame_Accumulated}"))
        if(PlayerData.HexaGame_Accumulated>=PlayerData.HexaGame_LVAccumulated+HexaGame_MinToken_Accumulated_MinusOne):
            HexaGameTimer_TextBlock.SetText(BasicMessage("FULL"))
        else if(PlayerData.NumQuestsCompleted>TutorialHexaGame_Index):
            HexaTimeTuple:=TransformToMinutes_Seconds_FromFloat(PlayerData.HexaGame_DelayTime)
            HexaGameTimer_TextBlock.SetText(BasicMessage("{HexaTimeTuple(0)}:{HexaTimeTuple(1)}"))
        else:
            HexaGameTimer_TextBlock.SetText(BasicMessage("--:--"))
            HexaGameNumTokens_TextBlock.SetText(BasicMessage("-"))
        #Dungeon
        DungeonNumTokens_TextBlock.SetText(BasicMessage("{PlayerData.Dungeon_Accumulated}"))
        if(PlayerData.Dungeon_Accumulated>=PlayerData.Dungeon_LVAccumulated):
            DungeonTimer_TextBlock.SetText(BasicMessage("FULL"))
        else if(PlayerData.NumQuestsCompleted>TutorialDungeon_Index):
            DungeonTimeTuple:=TransformToMinutes_Seconds_FromFloat(PlayerData.Dungeon_DelayTime)
            DungeonTimer_TextBlock.SetText(BasicMessage("{DungeonTimeTuple(0)}:{DungeonTimeTuple(1)}"))
        else:
            DungeonTimer_TextBlock.SetText(BasicMessage("--:--"))
            DungeonNumTokens_TextBlock.SetText(BasicMessage("-"))

        if(Agent:=MaybeAgent?):
            MaybeHexaManager:=BakeryMachinesManager.GameManager.GetAgentHexaManager(Agent)
            if:
                HexaManager:=MaybeHexaManager?
                HexMiniGameUI:=HexaManager.MaybeHexMiniGameUI?
                HexSelectorUI:=HexaManager.MaybeHexMiniGameSelectorUI?
                not HexMiniGameUI.UIIsDisplayed?
                not HexSelectorUI.UIIsDisplayed?
                not HexaManager.UIIsChallengeCanvas?
                not HexaManager.UIIsNavigationCanvas?
            then:
                #ORDER AND QUESTS
                if(not IsOrderCompleted_BeingDisplayed?, PlayerData.Order_DelayTime>0.0, PlayerData.NumQuestsCompleted>TutorialOrder_Index, PlayerData.OrderAssigned_ProductRequested=-99):
                    OrderTimeTuple:=TransformToMinutes_Seconds_FromFloat(PlayerData.Order_DelayTime)
                    OrderProgress_TextBlock.SetText(BasicMessage("{OrderTimeTuple(0)}:{OrderTimeTuple(1)}"))
                    OrderProgress_TextBlock.SetTextColor(Red)
                #else if(PlayerData.NumQuestCompleted>TutorialOrder_Index):
                #    OrderProgress_TextBlock.SetText(BasicMessage("--:--"))

                if(not IsQuestCompleted_BeingDisplayed?, PlayerData.Quest_DelayTime>0.0, PlayerData.NumQuestsCompleted>Quest0_Index):
                    QuestTimeTuple:=TransformToMinutes_Seconds_FromFloat(PlayerData.Quest_DelayTime)
                    QuestDescription_TextBlock.SetText(BasicMessage("{QuestTimeTuple(0)}:{QuestTimeTuple(1)}"))
                #else if(PlayerData.NumQuestCompleted>Quest0_Index):
                #    QuestDescription_TextBlock.SetText(BasicMessage("--:--"))

    RefreshInventory(PlayerData : player_data_table, AllLimits : []float, WasOverLimit : []logic) : void =
        Log("Refreshing UI!",LOG_LEVEL_UIs)
        if:
            WaterLimit := AllLimits[WATER]
            FlourLimit := AllLimits[FLOUR]
            CheeseLimit := AllLimits[CHEESE]
            PimentosLimit := AllLimits[PIMENTOS]

            WasOverLimit_Water:=WasOverLimit[WATER]
            WasOverLimit_Flour:=WasOverLimit[FLOUR]
            WasOverLimit_Cheese:=WasOverLimit[CHEESE]
            WasOverLimit_Pimentos:=WasOverLimit[PIMENTOS]
            not BakeryMachinesManager.HexaGameManager.IsUsingHexaGame?
            not BakeryMachinesManager.IsOnUKTutorial?
        then:
            #All time refresh
            RefreshResource(PlayerData.Inventory_Sapphires,-1.0,false,SAPPHIRES)
            RefreshResource(PlayerData.Inventory_Fame,-1.0,false,FAME)
            RefreshResource(PlayerData.Inventory_Water,WaterLimit,WasOverLimit_Water,WATER)
            RefreshResource(PlayerData.Inventory_Flour,FlourLimit,WasOverLimit_Flour,FLOUR)

            #RESOURCERS
            #Water
            BackpackWater_EmptyPack.SetDesiredSize(Limited_EmptyPack)
            BackpackWater_Icon.SetDesiredSize(VerySizeSmallIcon)
            BackpackWater_Border.SetDesiredSize(Limited_Border)

            #Flour
            BackpackFlour_EmptyPack.SetDesiredSize(Limited_EmptyPack)
            BackpackFlour_Icon.SetDesiredSize(VerySizeSmallIcon)
            BackpackFlour_Border.SetDesiredSize(Limited_Border)

            if(PlayerData.NumQuestsCompleted>1):
                #Cheese
                RefreshResource(PlayerData.Inventory_Cheese,CheeseLimit,WasOverLimit_Cheese,CHEESE)
                BackpackCheese_EmptyPack.SetDesiredSize(Limited_EmptyPack)
                BackpackCheese_Icon.SetDesiredSize(VerySizeSmallIcon)
                BackpackCheese_Border.SetDesiredSize(Limited_Border)
            if(PlayerData.NumQuestsCompleted>1):
                #Pimentos
                RefreshResource(PlayerData.Inventory_Pimentos,PimentosLimit,WasOverLimit_Pimentos,PIMENTOS)
                BackpackPimentos_EmptyPack.SetDesiredSize(Limited_EmptyPack)
                BackpackPimentos_Icon.SetDesiredSize(VerySizeSmallIcon)
                BackpackPimentos_Border.SetDesiredSize(Limited_Border)
            
            #Currencies
            BackpackSapphires_Background.SetDesiredSize(NotLimited_Background)
            BackpackSapphires_Border.SetDesiredSize(NotLimited_Border)
            BackpackSapphires_Icon.SetDesiredSize(VerySizeSmallIcon)

            if(PlayerData.NumQuestsCompleted>TutorialCoal_Index):
                #Coal
                RefreshResource(PlayerData.Inventory_Coal,-1.0,false,COAL)
                BackpackCoal_Background.SetDesiredSize(NotLimited_Background)
                BackpackCoal_Border.SetDesiredSize(NotLimited_Border)
                BackpackCoal_Icon.SetDesiredSize(VerySizeSmallIcon)
            if(PlayerData.NumQuestsCompleted>TutorialAggedSapphires_Index):
                #Aged Sapphires
                RefreshResource(PlayerData.Inventory_AgedSapphires,-1.0,false,AGEDSAPPHIRES)
                BackpackAgedSapphires_Background.SetDesiredSize(NotLimited_Background)
                BackpackAgedSapphires_Border.SetDesiredSize(NotLimited_Border)
                BackpackAgedSapphires_Icon.SetDesiredSize(VerySizeSmallIcon)
            if(PlayerData.NumQuestsCompleted>3):
                #Mutant Material
                RefreshResource(PlayerData.Inventory_MutantMaterial,-1.0,false,MUTANTMATERIAL)
                BackpackMutantMaterial_Background.SetDesiredSize(NotLimited_Background)
                BackpackMutantMaterial_Border.SetDesiredSize(NotLimited_Border)
                BackpackMutantMaterial_Icon.SetDesiredSize(VerySizeSmallIcon)

    RefreshOrders(PlayerData : player_data_table) : void =
        #Order
        Show_OrderPanel()
        if(not BakeryMachinesManager.HexaGameManager.IsUsingHexaGame?):
            DisplayOrder(PlayerData,PlayerData.OrderAssigned_ProductRequested,
                PlayerData.OrderAssigned_ProgressAmount,
                PlayerData.OrderAssigned_ProductAmountRequested,
                PlayerData.OrderAssigned_RewardType,
                PlayerData.OrderAssigned_Reward)
    RefreshQuests(PlayerData : player_data_table) : void =
        #Quest
        if(not BakeryMachinesManager.HexaGameManager.IsUsingHexaGame?, not BakeryMachinesManager.IsOnUKTutorial?):
            Show_QuestPanel()
            DisplayQuest(PlayerData)

    RefreshOrdersQuests(PlayerData : player_data_table) : void =
        if(not BakeryMachinesManager.HexaGameManager.IsUsingHexaGame?, not BakeryMachinesManager.IsOnUKTutorial?):
            #Order
            Show_OrderPanel()
            DisplayOrder(PlayerData,PlayerData.OrderAssigned_ProductRequested,
                PlayerData.OrderAssigned_ProgressAmount,
                PlayerData.OrderAssigned_ProductAmountRequested,
                PlayerData.OrderAssigned_RewardType,
                PlayerData.OrderAssigned_Reward)
            #Quest
            Show_QuestPanel()
            DisplayQuest(PlayerData)
        #Fame
        if(not IsFameHided?):
            Show_Fame(PlayerData)

    Display_QuestCompleted() : void =
        set IsQuestCompleted_BeingDisplayed = true
        QuestTitle_TextBlock.SetText(BasicMessage("Good job!"))
        QuestIcon_TextureBlock.SetDesiredSize(vector2{})
    
        QuestDescription_TextBlock.SetText(BasicMessage("Loading"))
        QuestDescription_TextBlock.SetTextColor(UpgradeTextColor)

        QuestProgress_TextBlock.SetText(BasicMessage("new"))
        QuestProgress_TextBlock.SetTextColor(UpgradeTextColor)

        QuestRewardDescrip_TextBlock.SetText(BasicMessage("quest..."))
        QuestRewardDescrip_TextBlock.SetTextColor(UpgradeTextColor)
        QuestReward_TextBlock.SetText(BasicMessage(""))
        QuestReward_TextBlock.SetTextColor(UpgradeTextColor)
        
        QuestReward_TextureBlock.SetDesiredSize(vector2{})
        spawn:
            QuestCompletedCeremony()

    QuestCompletedCeremony<private>()<suspends> : void =
        for(Index:=0..19):
            QuestDescription_TextBlock.SetTextColor(Red)
            QuestProgress_TextBlock.SetTextColor(Red)
            QuestRewardDescrip_TextBlock.SetTextColor(Red)
            Sleep(0.1)
            QuestDescription_TextBlock.SetTextColor(UpgradeTextColor)
            QuestProgress_TextBlock.SetTextColor(UpgradeTextColor)
            QuestRewardDescrip_TextBlock.SetTextColor(UpgradeTextColor)
            Sleep(0.1)


    Display_OrderCompleted() : void =
        set IsOrderCompleted_BeingDisplayed = true
        OrderTitle_TextBlock.SetText(BasicMessage("Good job!"))

        OrderProduct_TextureBlock.SetDesiredSize(vector2{})
        OrderProgress_TextBlock.SetText(BasicMessage("Loading next"))
        OrderProgress_TextBlock.SetTextColor(UpgradeTextColor)

        OrderReward_TextureBlock.SetDesiredSize(vector2{})
        OrderReward_TextBlock.SetText(BasicMessage("order..."))
        OrderReward_TextBlock.SetTextColor(UpgradeTextColor)

        OrderRewardDescrip_TextBlock.SetText(BasicMessage(""))

    RefreshAll(PlayerData : player_data_table, AllLimits : []float, WasOverLimit : []logic) : void =
        Show_Fame(PlayerData)
        if(not BakeryMachinesManager.HexaGameManager.IsUsingHexaGame?):
            RefreshTimers(PlayerData)
            RefreshInventory(PlayerData,AllLimits,WasOverLimit)
            RefreshOrdersQuests(PlayerData)

    SimplifiedRefreshAll() : void =
        if(not BakeryMachinesManager.HexaGameManager.IsUsingHexaGame?):
            Show_Fame(BakeryMachinesManager.PlayerDataCache)
            RefreshTimers(BakeryMachinesManager.PlayerDataCache)
            BakeryMachinesManager.RefreshInventoryUI(BakeryMachinesManager.PlayerDataCache)
            RefreshOrdersQuests(BakeryMachinesManager.PlayerDataCache)

    Hide_Fame() : void =
        set IsFameHided = true
        OptionsInfo_Texture.SetDesiredSize(vector2{})
        BackpackFame_Icon.SetDesiredSize(vector2{})
        BackpackFame_Amount.SetText(BasicMessage(""))

    Show_Fame(PlayerData : player_data_table) : void =
        set IsFameHided = false
        OptionsInfo_Texture.SetDesiredSize(OptionsInfoPanelSize)
        BackpackFame_Icon.SetDesiredSize(VerySizeSmallIcon)
        BackpackFame_Amount.SetText(BasicMessage(GetOptimizedMagnitudeLetter(PlayerData.Inventory_Fame)))
        #FameUI_Slider.SetDesiredSize(vector2{X:=605.0,Y:=0.0})

    RunCharacterMessage<private>(DisplayTime : float, IsLocalized : logic)<suspends> : void =
        if(not IsLocalized?):
            CharacterMessage_Slider.SetDesiredSize(vector2{X:=5.0,Y:=CharacterMessage_SliderDisplayedAmount})
            Sleep(DisplayTime)   
            CharacterMessage_Slider.SetDesiredSize(vector2{X:=5.0,Y:=CharacterMessage_SliderNotDisplayedAmount})
        else:
            CharacterMessage_SliderLoc.SetDesiredSize(vector2{X:=5.0,Y:=CharacterMessage_SliderDisplayedAmount})
            Sleep(DisplayTime)   
            CharacterMessage_SliderLoc.SetDesiredSize(vector2{X:=5.0,Y:=CharacterMessage_SliderNotDisplayedAmount})

    Remove_Canvas(Message : widget_message) : void =
        set UIIsDisplayed = false
        ##Print("Trying to remove UI")
        if:
            Agent := MaybeAgent?
            Player:=player[Agent]
            PlayerUI:=GetPlayerUI[Player]
            Canvas := MaybeCanvas?
        then:
        #    #Print("Removing UI")
            PlayerUI.RemoveWidget(Canvas)
     

    Remove_CanvasWithoutMessage() : void =
        set UIIsDisplayed = false
        ##Print("Trying to remove UI")
        if:
            Agent := MaybeAgent?
            Player:=player[Agent]
            PlayerUI:=GetPlayerUI[Player]
            Canvas := MaybeCanvas?
        then:
        #    #Print("Removing UI")
            PlayerUI.RemoveWidget(Canvas)
  

            
    Display_Canvas() : void =
        if(not UIIsDisplayed?):
            set UIIsDisplayed = true
            var Index : int = 0 
            if:
                Agent := MaybeAgent?
                Player:=player[Agent]
                PlayerUI:=GetPlayerUI[player[Agent]]
                Canvas := MaybeCanvas?
            then:
                ##Print("Creating canvas for the first time")
                #PrintWarning("TO DO on Display_Canvas() of main_inventory_ui calculate backpack limits")
                PlayerUI.AddWidget(Canvas,player_ui_slot{ZOrder:=0})
                Hide_ResourcersPanel()
                RefreshAll(BakeryMachinesManager.PlayerDataCache, array{999.9,999.9,999.9,999.9}, array{false,false,false,false})
   

    MakeCanvas(): canvas=
        Canvas := canvas:
            Slots:=array:
                #Options info
                canvas_slot:
                    Anchors:=anchors{Minimum:=vector2{X:=0.5,Y:=0.0},Maximum:=vector2{X:=0.5,Y:=0.0}}
                    ZOrder:=0
                    SizeToContent:=true
                    Offsets:=margin{Left:=500.0,Top:=0.0,Right:=0.0,Bottom:=0.0} #PositionX, PositionY, SizeX, SizeY
                    Alignment:=vector2{X:=0.5,Y:=0.0}
                    Widget:=overlay:
                        Slots:=array:
                            overlay_slot:
                                HorizontalAlignment:=horizontal_alignment.Center
                                VerticalAlignment:=vertical_alignment.Center
                                Widget:=OptionsInfo_Texture

                #DamageEffect
                canvas_slot:
                    Anchors:=anchors{Minimum:=vector2{X:=0.5,Y:=0.5},Maximum:=vector2{X:=0.5,Y:=0.5}}
                    ZOrder:=0
                    SizeToContent:=true
                    Offsets:=margin{Left:=0.0,Top:=0.0,Right:=0.0,Bottom:=0.0} #PositionX, PositionY, SizeX, SizeY
                    Alignment:=vector2{X:=0.5,Y:=0.5}
                    Widget:=DamageEffect_Texture
                #Quests dialogue
                canvas_slot:
                    Anchors:=anchors{Minimum:=vector2{X:=1.0,Y:=0.5},Maximum:=vector2{X:=1.0,Y:=0.5}}
                    ZOrder:=0
                    SizeToContent:=true
                    Offsets:=margin{Left:=0.0,Top:=-50.0,Right:=0.0,Bottom:=0.0} #PositionX, PositionY, SizeX, SizeY
                    Alignment:=vector2{X:=1.0,Y:=0.5}
                    Widget:=overlay:
                        Slots:=array:
                            overlay_slot:
                                HorizontalAlignment:=horizontal_alignment.Center
                                VerticalAlignment:=vertical_alignment.Center
                                Widget:=overlay:
                                    Slots:=array:
                                        overlay_slot:
                                            HorizontalAlignment:=horizontal_alignment.Fill
                                            VerticalAlignment:=vertical_alignment.Fill
                                            Widget:=Quest_Background0
                                        overlay_slot:
                                            HorizontalAlignment:=horizontal_alignment.Fill
                                            VerticalAlignment:=vertical_alignment.Fill
                                            Padding:=margin{Left:=2.5,Top:=2.5,Right:=2.5,Bottom:=2.5}
                                            Widget:=Quest_Background1
                                        overlay_slot: #Screw 0
                                            Padding:=margin{Left:=3.5,Top:=3.5,Right:=3.5,Bottom:=3.5}
                                            HorizontalAlignment:=horizontal_alignment.Left
                                            VerticalAlignment:=vertical_alignment.Top
                                            Widget:=Quest_Bolt0
                                        overlay_slot: #Screw 1
                                            Padding:=margin{Left:=3.5,Top:=3.5,Right:=3.5,Bottom:=3.5}
                                            HorizontalAlignment:=horizontal_alignment.Right
                                            VerticalAlignment:=vertical_alignment.Top
                                            Widget:=Quest_Bolt1
                                        overlay_slot: #Screw 2
                                            Padding:=margin{Left:=3.5,Top:=3.5,Right:=3.5,Bottom:=3.5}
                                            HorizontalAlignment:=horizontal_alignment.Left
                                            VerticalAlignment:=vertical_alignment.Bottom
                                            Widget:=Quest_Bolt2
                                        overlay_slot: #Screw 3
                                            Padding:=margin{Left:=3.5,Top:=3.5,Right:=3.5,Bottom:=3.5}
                                            HorizontalAlignment:=horizontal_alignment.Right
                                            VerticalAlignment:=vertical_alignment.Bottom
                                            Widget:=Quest_Bolt3

                                        overlay_slot: #Content
                                            HorizontalAlignment:=horizontal_alignment.Center
                                            VerticalAlignment:=vertical_alignment.Center
                                            Padding:=margin{Left:=5.0,Top:=5.0,Right:=5.0,Bottom:=5.0}
                                            Widget:=stack_box:
                                                Orientation:=orientation.Vertical
                                                    Slots:=array:
                                                        stack_box_slot:
                                                            Widget:=stack_box: #First line
                                                                Orientation:=orientation.Horizontal
                                                                    Slots:=array:
                                                                        stack_box_slot:
                                                                            Padding:=margin{Left:=2.5,Top:=2.5,Right:=2.5,Bottom:=2.5}
                                                                            Widget:=QuestIcon_TextureBlock
                                                                        stack_box_slot:
                                                                            Padding:=margin{Left:=2.5,Top:=2.5,Right:=2.5,Bottom:=2.5}
                                                                            Widget:=QuestTitle_TextBlock
                                                        stack_box_slot:
                                                            Widget:=stack_box: #Second line
                                                                Orientation:=orientation.Horizontal
                                                                    Slots:=array:
                                                                        stack_box_slot:
                                                                            Padding:=margin{Left:=2.5,Top:=2.5,Right:=2.5,Bottom:=2.5}
                                                                            Widget:=QuestDescription_TextBlock
                                                        stack_box_slot:
                                                            Widget:=stack_box: #Third line
                                                                Orientation:=orientation.Horizontal
                                                                    Slots:=array:
                                                                        stack_box_slot:
                                                                            Padding:=margin{Left:=2.5,Top:=2.5,Right:=2.5,Bottom:=2.5}
                                                                            Widget:=QuestProgress_TextBlock
                                                        stack_box_slot:
                                                            Widget:=stack_box: #Fourth line
                                                                Orientation:=orientation.Horizontal
                                                                    Slots:=array:
                                                                        stack_box_slot:
                                                                            Padding:=margin{Left:=2.5,Top:=2.5,Right:=2.5}
                                                                            Widget:=QuestRewardDescrip_TextBlock
                                                                        stack_box_slot:
                                                                            Padding:=margin{Left:=2.5,Top:=2.5,Right:=2.5,Bottom:=2.5}
                                                                            Widget:=QuestReward_TextureBlock
                                                                        stack_box_slot:
                                                                            Padding:=margin{Left:=2.5,Top:=2.5,Right:=2.5,Bottom:=2.5}
                                                                            Widget:=QuestReward_TextBlock
                #Orders dialogue
                canvas_slot:
                    Anchors:=anchors{Minimum:=vector2{X:=1.0,Y:=0.5},Maximum:=vector2{X:=1.0,Y:=0.5}}
                    ZOrder:=0
                    SizeToContent:=true
                    Offsets:=margin{Left:=0.0,Top:=220.0,Right:=0.0,Bottom:=0.0} #PositionX, PositionY, SizeX, SizeY
                    Alignment:=vector2{X:=1.0,Y:=0.5}
                    Widget:=overlay:
                        Slots:=array:
                            overlay_slot:
                                HorizontalAlignment:=horizontal_alignment.Center
                                VerticalAlignment:=vertical_alignment.Center
                                Widget:=overlay:
                                    Slots:=array:
                                        overlay_slot:
                                            HorizontalAlignment:=horizontal_alignment.Fill
                                            VerticalAlignment:=vertical_alignment.Fill
                                            Widget:=Order_Background0
                                        overlay_slot:
                                            HorizontalAlignment:=horizontal_alignment.Fill
                                            VerticalAlignment:=vertical_alignment.Fill
                                            Padding:=margin{Left:=2.5,Top:=2.5,Right:=2.5,Bottom:=2.5}
                                            Widget:=Order_Background1
                                        overlay_slot: #Screw 0
                                            Padding:=margin{Left:=3.5,Top:=3.5,Right:=3.5,Bottom:=3.5}
                                            HorizontalAlignment:=horizontal_alignment.Left
                                            VerticalAlignment:=vertical_alignment.Top
                                            Widget:=Order_Bolt0
                                        overlay_slot: #Screw 1
                                            Padding:=margin{Left:=3.5,Top:=3.5,Right:=3.5,Bottom:=3.5}
                                            HorizontalAlignment:=horizontal_alignment.Right
                                            VerticalAlignment:=vertical_alignment.Top
                                            Widget:=Order_Bolt1
                                        overlay_slot: #Screw 2
                                            Padding:=margin{Left:=3.5,Top:=3.5,Right:=3.5,Bottom:=3.5}
                                            HorizontalAlignment:=horizontal_alignment.Left
                                            VerticalAlignment:=vertical_alignment.Bottom
                                            Widget:=Order_Bolt2
                                        overlay_slot: #Screw 3
                                            Padding:=margin{Left:=3.5,Top:=3.5,Right:=3.5,Bottom:=3.5}
                                            HorizontalAlignment:=horizontal_alignment.Right
                                            VerticalAlignment:=vertical_alignment.Bottom
                                            Widget:=Order_Bolt3

                                        overlay_slot: #Content
                                            HorizontalAlignment:=horizontal_alignment.Center
                                            VerticalAlignment:=vertical_alignment.Center
                                            Padding:=margin{Left:=5.0,Top:=5.0,Right:=5.0,Bottom:=5.0}
                                            Widget:=stack_box:
                                                Orientation:=orientation.Vertical
                                                    Slots:=array:
                                                        stack_box_slot:
                                                            Widget:=stack_box: #First line
                                                                Orientation:=orientation.Horizontal
                                                                    Slots:=array:
                                                                        stack_box_slot:
                                                                            Padding:=margin{Left:=2.5,Top:=2.5,Right:=2.5,Bottom:=2.5}
                                                                            Widget:=OrderTitle_TextBlock
                                                                        stack_box_slot:
                                                                            Padding:=margin{Left:=2.5,Top:=2.5,Right:=2.5,Bottom:=2.5}
                                                                            Widget:=OrderProduct_TextureBlock
                                                        stack_box_slot:
                                                            Widget:=stack_box: #Second line
                                                                Orientation:=orientation.Horizontal
                                                                    Slots:=array:
                                                                        stack_box_slot:
                                                                            Padding:=margin{Left:=2.5,Top:=2.5,Right:=2.5,Bottom:=2.5}
                                                                            Widget:=OrderProgress_TextBlock

                                                        stack_box_slot:
                                                            Widget:=stack_box: #Third line
                                                                Orientation:=orientation.Horizontal
                                                                    Slots:=array:
                                                                        stack_box_slot:
                                                                            Padding:=margin{Left:=2.5,Top:=2.5,Right:=2.5}
                                                                            Widget:=OrderRewardDescrip_TextBlock
                                                                        stack_box_slot:
                                                                            Padding:=margin{Left:=2.5,Right:=2.5}
                                                                            Widget:=OrderReward_TextureBlock
                                                                        stack_box_slot:
                                                                            Padding:=margin{Left:=2.5,Right:=2.5,Bottom:=2.5}
                                                                            Widget:=OrderReward_TextBlock

                #Bottom characters dialogue
                canvas_slot:
                    Anchors:=anchors{Minimum:=vector2{X:=0.5,Y:=1.0},Maximum:=vector2{X:=0.5,Y:=1.0}}
                    ZOrder:=0
                    SizeToContent:=true
                    Offsets:=margin{Left:=0.0,Top:=-145.0,Right:=0.0,Bottom:=0.0} #PositionX, PositionY, SizeX, SizeY
                    Alignment:=vector2{X:=0.5,Y:=0.5}
                    Widget:=overlay:
                        Slots:=array:
                            overlay_slot:
                                HorizontalAlignment:=horizontal_alignment.Center
                                VerticalAlignment:=vertical_alignment.Center
                                Widget:=CharacterMessage_Slider
                            overlay_slot:
                                HorizontalAlignment:=horizontal_alignment.Center
                                VerticalAlignment:=vertical_alignment.Bottom
                                Widget:=overlay:
                                    Slots:=array:
                                        overlay_slot:
                                            HorizontalAlignment:=horizontal_alignment.Left
                                            VerticalAlignment:=vertical_alignment.Center
                                            Padding:=margin{Left:=19.0}
                                            Widget:=CharacterImage_TextureBlock
                                        overlay_slot:
                                            HorizontalAlignment:=horizontal_alignment.Center
                                            VerticalAlignment:=vertical_alignment.Center
                                            Widget:=CharacterMessage_TextureBlock
                                        overlay_slot:
                                            HorizontalAlignment:=horizontal_alignment.Left
                                            VerticalAlignment:=vertical_alignment.Center
                                            Padding:=margin{Left:=175.0}
                                            Widget:=CharacterMessage_TextBlock         
                #Bottom characters dialogue LOCALICED
                canvas_slot:
                    Anchors:=anchors{Minimum:=vector2{X:=0.5,Y:=1.0},Maximum:=vector2{X:=0.5,Y:=1.0}}
                    ZOrder:=0
                    SizeToContent:=true
                    Offsets:=margin{Left:=0.0,Top:=-145.0,Right:=0.0,Bottom:=0.0} #PositionX, PositionY, SizeX, SizeY
                    Alignment:=vector2{X:=0.5,Y:=0.5}
                    Widget:=overlay:
                        Slots:=array:
                            overlay_slot:
                                HorizontalAlignment:=horizontal_alignment.Center
                                VerticalAlignment:=vertical_alignment.Center
                                Widget:=CharacterMessage_SliderLoc
                            overlay_slot:
                                HorizontalAlignment:=horizontal_alignment.Center
                                VerticalAlignment:=vertical_alignment.Bottom
                                Widget:=overlay:
                                    Slots:=array:
                                        overlay_slot:
                                            HorizontalAlignment:=horizontal_alignment.Left
                                            VerticalAlignment:=vertical_alignment.Center
                                            Padding:=margin{Left:=10.0}#{Left:=19.0}
                                            Widget:=CharacterImage_TextureBlockLoc
                                        overlay_slot:
                                            HorizontalAlignment:=horizontal_alignment.Center
                                            VerticalAlignment:=vertical_alignment.Center
                                            Widget:=CharacterMessage_TextureBlockLoc
                                        overlay_slot:
                                            HorizontalAlignment:=horizontal_alignment.Left
                                            VerticalAlignment:=vertical_alignment.Center
                                            Padding:=margin{Left:=150.0}#{Left:=175.0}
                                            Widget:=CharacterMessage_TextBlockLoc      
                #Fame top overlay
                canvas_slot:
                    Anchors:=anchors{Minimum:=vector2{X:=0.5,Y:=0.0},Maximum:=vector2{X:=0.5,Y:=0.0}}
                    ZOrder:=0
                    SizeToContent:=true
                    Offsets:=margin{Left:=0.0,Top:=0.0,Right:=0.0,Bottom:=0.0} #PositionX, PositionY, SizeX, SizeY
                    Alignment:=vector2{X:=0.5,Y:=0.0}
                    Widget:=overlay:
                        Slots:=array:
                            overlay_slot:
                                Padding:=margin{Top:=5.0}
                                Widget:=stack_box:
                                    Orientation:=orientation.Horizontal 
                                        Slots:=array:
                                            stack_box_slot: 
                                                HorizontalAlignment:=horizontal_alignment.Center
                                                VerticalAlignment:=vertical_alignment.Center
                                                Widget:=BackpackFame_Icon
                                            stack_box_slot:
                                                Padding:=margin{Left:=10.0}
                                                HorizontalAlignment:=horizontal_alignment.Center
                                                VerticalAlignment:=vertical_alignment.Bottom 
                                                Widget:=BackpackFame_Amount
                                
                #Timers & pet perk soverlay
                canvas_slot:
                    Anchors:=anchors{Minimum:=vector2{X:=0.0,Y:=1.0},Maximum:=vector2{X:=0.0,Y:=1.0}}
                    ZOrder:=0
                    SizeToContent:=true
                    Offsets:=margin{Left:=5.0,Top:=-5.0,Right:=0.0,Bottom:=0.0} #PositionX, PositionY, SizeX, SizeY
                    Alignment:=vector2{X:=0.0,Y:=1.0}
                    Widget:=stack_box:
                        Orientation:=orientation.Horizontal
                            Slots:=array:
                                stack_box_slot: 
                                    Widget:=overlay:
                                        Slots:=array:
                                            overlay_slot:
                                                HorizontalAlignment:=horizontal_alignment.Fill
                                                VerticalAlignment:=vertical_alignment.Fill
                                                Widget:=texture_block{DefaultImage:=Images.Img_LineBackground}
                                            overlay_slot:
                                                HorizontalAlignment:=horizontal_alignment.Fill
                                                VerticalAlignment:=vertical_alignment.Fill
                                                Padding:=margin{Left:=InventorySliderBarPadding,Top:=InventorySliderBarPadding,Right:=InventorySliderBarPadding,Bottom:=InventorySliderBarPadding}
                                                Widget:=texture_block{DefaultDesiredSize:=vector2{X:=InventoryBoxSizeX,Y:=InventoryBoxSizeY},DefaultImage:=Images.Img_LineBackground}
                                            overlay_slot: #Screw 0
                                                Padding:=margin{Left:=3.5,Top:=3.5,Right:=3.5,Bottom:=3.5}
                                                HorizontalAlignment:=horizontal_alignment.Left
                                                VerticalAlignment:=vertical_alignment.Top
                                                Widget:=texture_block{DefaultImage:=Images.Dialogue_Bolt, DefaultDesiredSize:=vector2{X:=5.0,Y:=5.0}}
                                            overlay_slot: #Screw 1
                                                Padding:=margin{Left:=3.5,Top:=3.5,Right:=3.5,Bottom:=3.5}
                                                HorizontalAlignment:=horizontal_alignment.Right
                                                VerticalAlignment:=vertical_alignment.Top
                                                Widget:=texture_block{DefaultImage:=Images.Dialogue_Bolt, DefaultDesiredSize:=vector2{X:=5.0,Y:=5.0}}
                                            overlay_slot: #Screw 2
                                                Padding:=margin{Left:=3.5,Top:=3.5,Right:=3.5,Bottom:=3.5}
                                                HorizontalAlignment:=horizontal_alignment.Left
                                                VerticalAlignment:=vertical_alignment.Bottom
                                                Widget:=texture_block{DefaultImage:=Images.Dialogue_Bolt, DefaultDesiredSize:=vector2{X:=5.0,Y:=5.0}}
                                            overlay_slot: #Screw 3
                                                Padding:=margin{Left:=3.5,Top:=3.5,Right:=3.5,Bottom:=3.5}
                                                HorizontalAlignment:=horizontal_alignment.Right
                                                VerticalAlignment:=vertical_alignment.Bottom
                                                Widget:=texture_block{DefaultImage:=Images.Dialogue_Bolt, DefaultDesiredSize:=vector2{X:=5.0,Y:=5.0}}
                                            overlay_slot:
                                                Widget:=stack_box:
                                                    Orientation:=orientation.Horizontal
                                                    Slots:=array:
                                                        stack_box_slot:
                                                            Padding:=margin{Left:=InventorySliderBarPadding,Right:=InventorySliderBarPadding}
                                                            Widget:=overlay:
                                                                Slots:=array:
                                                                    overlay_slot:
                                                                        Widget:=texture_block{DefaultDesiredSize:=vector2{X:=VerySizeSmallIconX,Y:=VerySizeSmallIconY},DefaultImage:=Images.HexTrophy}
                                                        stack_box_slot:
                                                            Padding:=margin{Left:=InventorySliderBarPadding,Right:=InventorySliderBarPadding}
                                                            Widget:=overlay:
                                                                Slots:=array:
                                                                    overlay_slot:
                                                                        Widget:=texture_block{DefaultDesiredSize:=vector2{X:=50.0,Y:=50.0},DefaultImage:=Images.BlackBackground}
                                                                    overlay_slot:
                                                                        Widget:=HexaGameNumTokens_TextBlock
                                                        stack_box_slot:
                                                            Padding:=margin{Left:=InventorySliderBarPadding,Right:=InventorySliderBarPadding}
                                                            Widget:=overlay:
                                                                Slots:=array:
                                                                    overlay_slot:
                                                                        Widget:=texture_block{DefaultDesiredSize:=vector2{X:=100.0,Y:=50.0},DefaultImage:=Images.EmptyTexture}
                                                                    overlay_slot:
                                                                        Widget:=HexaGameTimer_TextBlock
                                stack_box_slot: 
                                    Padding:=margin{Left:=5.0}
                                    Widget:=overlay:
                                        Slots:=array:
                                            overlay_slot:
                                                HorizontalAlignment:=horizontal_alignment.Fill
                                                VerticalAlignment:=vertical_alignment.Fill
                                                Widget:=texture_block{DefaultImage:=Images.Img_LineBackground}
                                            overlay_slot:
                                                HorizontalAlignment:=horizontal_alignment.Fill
                                                VerticalAlignment:=vertical_alignment.Fill
                                                Padding:=margin{Left:=InventorySliderBarPadding,Top:=InventorySliderBarPadding,Right:=InventorySliderBarPadding,Bottom:=InventorySliderBarPadding}
                                                Widget:=texture_block{DefaultDesiredSize:=vector2{X:=InventoryBoxSizeX,Y:=InventoryBoxSizeY},DefaultImage:=Images.Img_LineBackground}
                                            overlay_slot: #Screw 0
                                                Padding:=margin{Left:=3.5,Top:=3.5,Right:=3.5,Bottom:=3.5}
                                                HorizontalAlignment:=horizontal_alignment.Left
                                                VerticalAlignment:=vertical_alignment.Top
                                                Widget:=texture_block{DefaultImage:=Images.Dialogue_Bolt, DefaultDesiredSize:=vector2{X:=5.0,Y:=5.0}}
                                            overlay_slot: #Screw 1
                                                Padding:=margin{Left:=3.5,Top:=3.5,Right:=3.5,Bottom:=3.5}
                                                HorizontalAlignment:=horizontal_alignment.Right
                                                VerticalAlignment:=vertical_alignment.Top
                                                Widget:=texture_block{DefaultImage:=Images.Dialogue_Bolt, DefaultDesiredSize:=vector2{X:=5.0,Y:=5.0}}
                                            overlay_slot: #Screw 2
                                                Padding:=margin{Left:=3.5,Top:=3.5,Right:=3.5,Bottom:=3.5}
                                                HorizontalAlignment:=horizontal_alignment.Left
                                                VerticalAlignment:=vertical_alignment.Bottom
                                                Widget:=texture_block{DefaultImage:=Images.Dialogue_Bolt, DefaultDesiredSize:=vector2{X:=5.0,Y:=5.0}}
                                            overlay_slot: #Screw 3
                                                Padding:=margin{Left:=3.5,Top:=3.5,Right:=3.5,Bottom:=3.5}
                                                HorizontalAlignment:=horizontal_alignment.Right
                                                VerticalAlignment:=vertical_alignment.Bottom
                                                Widget:=texture_block{DefaultImage:=Images.Dialogue_Bolt, DefaultDesiredSize:=vector2{X:=5.0,Y:=5.0}}
                                            overlay_slot:
                                                Widget:=stack_box:
                                                    Orientation:=orientation.Horizontal
                                                    Slots:=array:
                                                        stack_box_slot:
                                                            Padding:=margin{Left:=InventorySliderBarPadding,Right:=InventorySliderBarPadding}
                                                            Widget:=overlay:
                                                                Slots:=array:
                                                                    overlay_slot:
                                                                        Widget:=texture_block{DefaultDesiredSize:=vector2{X:=VerySizeSmallIconX,Y:=VerySizeSmallIconY},DefaultImage:=Images.Dungeon_Icon}
                                                        stack_box_slot:
                                                            Padding:=margin{Left:=InventorySliderBarPadding,Right:=InventorySliderBarPadding}
                                                            Widget:=overlay:
                                                                Slots:=array:
                                                                    overlay_slot:
                                                                        Widget:=texture_block{DefaultDesiredSize:=vector2{X:=50.0,Y:=50.0},DefaultImage:=Images.BlackBackground}
                                                                    overlay_slot:
                                                                        Widget:=DungeonNumTokens_TextBlock
                                                        stack_box_slot:
                                                            Padding:=margin{Left:=InventorySliderBarPadding,Right:=InventorySliderBarPadding}
                                                            Widget:=overlay:
                                                                Slots:=array:
                                                                    overlay_slot:
                                                                        Widget:=texture_block{DefaultDesiredSize:=vector2{X:=100.0,Y:=50.0},DefaultImage:=Images.EmptyTexture}
                                                                    overlay_slot:
                                                                        Widget:=DungeonTimer_TextBlock
                                stack_box_slot: #PetPerkslot 0
                                    Padding:=margin{Left:=5.0}
                                    Widget:=overlay:
                                        Slots:=array:
                                            overlay_slot:
                                                HorizontalAlignment:=horizontal_alignment.Fill
                                                VerticalAlignment:=vertical_alignment.Fill
                                                Widget:=texture_block{DefaultImage:=Images.Img_LineBackground}
                                            overlay_slot:
                                                HorizontalAlignment:=horizontal_alignment.Fill
                                                VerticalAlignment:=vertical_alignment.Fill
                                                Padding:=margin{Left:=InventorySliderBarPadding,Top:=InventorySliderBarPadding,Right:=InventorySliderBarPadding,Bottom:=InventorySliderBarPadding}
                                                Widget:=texture_block{DefaultDesiredSize:=vector2{X:=PetPerkBoxSizeX,Y:=PetPerkBoxSizeY},DefaultImage:=Images.Img_LineBackground}
                                            overlay_slot:
                                                Padding:=margin{Left:=InventorySliderBarPadding,Top:=InventorySliderBarPadding,Right:=InventorySliderBarPadding,Bottom:=InventorySliderBarPadding}
                                                HorizontalAlignment:=horizontal_alignment.Left
                                                VerticalAlignment:=vertical_alignment.Fill
                                                Widget:=PetPerkSlider0_Texture
                                            overlay_slot: #Screw 0
                                                Padding:=margin{Left:=3.5,Top:=3.5,Right:=3.5,Bottom:=3.5}
                                                HorizontalAlignment:=horizontal_alignment.Left
                                                VerticalAlignment:=vertical_alignment.Top
                                                Widget:=texture_block{DefaultImage:=Images.Dialogue_Bolt, DefaultDesiredSize:=vector2{X:=5.0,Y:=5.0}}
                                            overlay_slot: #Screw 1
                                                Padding:=margin{Left:=3.5,Top:=3.5,Right:=3.5,Bottom:=3.5}
                                                HorizontalAlignment:=horizontal_alignment.Right
                                                VerticalAlignment:=vertical_alignment.Top
                                                Widget:=texture_block{DefaultImage:=Images.Dialogue_Bolt, DefaultDesiredSize:=vector2{X:=5.0,Y:=5.0}}
                                            overlay_slot: #Screw 2
                                                Padding:=margin{Left:=3.5,Top:=3.5,Right:=3.5,Bottom:=3.5}
                                                HorizontalAlignment:=horizontal_alignment.Left
                                                VerticalAlignment:=vertical_alignment.Bottom
                                                Widget:=texture_block{DefaultImage:=Images.Dialogue_Bolt, DefaultDesiredSize:=vector2{X:=5.0,Y:=5.0}}
                                            overlay_slot: #Screw 3
                                                Padding:=margin{Left:=3.5,Top:=3.5,Right:=3.5,Bottom:=3.5}
                                                HorizontalAlignment:=horizontal_alignment.Right
                                                VerticalAlignment:=vertical_alignment.Bottom
                                                Widget:=texture_block{DefaultImage:=Images.Dialogue_Bolt, DefaultDesiredSize:=vector2{X:=5.0,Y:=5.0}}
                                            overlay_slot:
                                                Padding:=margin{Left:=InventorySliderBarPadding,Top:=InventorySliderBarPadding,Right:=InventorySliderBarPadding,Bottom:=InventorySliderBarPadding}
                                                Widget:=PetPerk0_Text

                                stack_box_slot: #PetPerkslot 1
                                    Padding:=margin{Left:=5.0}
                                    Widget:=overlay:
                                        Slots:=array:
                                            overlay_slot:
                                                HorizontalAlignment:=horizontal_alignment.Fill
                                                VerticalAlignment:=vertical_alignment.Fill
                                                Widget:=texture_block{DefaultImage:=Images.Img_LineBackground}
                                            overlay_slot:
                                                HorizontalAlignment:=horizontal_alignment.Fill
                                                VerticalAlignment:=vertical_alignment.Fill
                                                Padding:=margin{Left:=InventorySliderBarPadding,Top:=InventorySliderBarPadding,Right:=InventorySliderBarPadding,Bottom:=InventorySliderBarPadding}
                                                Widget:=texture_block{DefaultDesiredSize:=vector2{X:=PetPerkBoxSizeX,Y:=PetPerkBoxSizeY},DefaultImage:=Images.Img_LineBackground}
                                            overlay_slot:
                                                Padding:=margin{Left:=InventorySliderBarPadding,Top:=InventorySliderBarPadding,Right:=InventorySliderBarPadding,Bottom:=InventorySliderBarPadding}
                                                HorizontalAlignment:=horizontal_alignment.Left
                                                VerticalAlignment:=vertical_alignment.Fill
                                                Widget:=PetPerkSlider1_Texture
                                            overlay_slot: #Screw 0
                                                Padding:=margin{Left:=3.5,Top:=3.5,Right:=3.5,Bottom:=3.5}
                                                HorizontalAlignment:=horizontal_alignment.Left
                                                VerticalAlignment:=vertical_alignment.Top
                                                Widget:=texture_block{DefaultImage:=Images.Dialogue_Bolt, DefaultDesiredSize:=vector2{X:=5.0,Y:=5.0}}
                                            overlay_slot: #Screw 1
                                                Padding:=margin{Left:=3.5,Top:=3.5,Right:=3.5,Bottom:=3.5}
                                                HorizontalAlignment:=horizontal_alignment.Right
                                                VerticalAlignment:=vertical_alignment.Top
                                                Widget:=texture_block{DefaultImage:=Images.Dialogue_Bolt, DefaultDesiredSize:=vector2{X:=5.0,Y:=5.0}}
                                            overlay_slot: #Screw 2
                                                Padding:=margin{Left:=3.5,Top:=3.5,Right:=3.5,Bottom:=3.5}
                                                HorizontalAlignment:=horizontal_alignment.Left
                                                VerticalAlignment:=vertical_alignment.Bottom
                                                Widget:=texture_block{DefaultImage:=Images.Dialogue_Bolt, DefaultDesiredSize:=vector2{X:=5.0,Y:=5.0}}
                                            overlay_slot: #Screw 3
                                                Padding:=margin{Left:=3.5,Top:=3.5,Right:=3.5,Bottom:=3.5}
                                                HorizontalAlignment:=horizontal_alignment.Right
                                                VerticalAlignment:=vertical_alignment.Bottom
                                                Widget:=texture_block{DefaultImage:=Images.Dialogue_Bolt, DefaultDesiredSize:=vector2{X:=5.0,Y:=5.0}}
                                            overlay_slot:
                                                Padding:=margin{Left:=InventorySliderBarPadding,Top:=InventorySliderBarPadding,Right:=InventorySliderBarPadding,Bottom:=InventorySliderBarPadding}
                                                Widget:=PetPerk1_Text

                                stack_box_slot: #PetPerkslot 2
                                    Padding:=margin{Left:=5.0}
                                    Widget:=overlay:
                                        Slots:=array:
                                            overlay_slot:
                                                HorizontalAlignment:=horizontal_alignment.Fill
                                                VerticalAlignment:=vertical_alignment.Fill
                                                Widget:=texture_block{DefaultImage:=Images.Img_LineBackground}
                                            overlay_slot:
                                                HorizontalAlignment:=horizontal_alignment.Fill
                                                VerticalAlignment:=vertical_alignment.Fill
                                                Padding:=margin{Left:=InventorySliderBarPadding,Top:=InventorySliderBarPadding,Right:=InventorySliderBarPadding,Bottom:=InventorySliderBarPadding}
                                                Widget:=texture_block{DefaultDesiredSize:=vector2{X:=PetPerkBoxSizeX,Y:=PetPerkBoxSizeY},DefaultImage:=Images.Img_LineBackground}
                                            overlay_slot:
                                                Padding:=margin{Left:=InventorySliderBarPadding,Top:=InventorySliderBarPadding,Right:=InventorySliderBarPadding,Bottom:=InventorySliderBarPadding}
                                                HorizontalAlignment:=horizontal_alignment.Left
                                                VerticalAlignment:=vertical_alignment.Fill
                                                Widget:=PetPerkSlider2_Texture
                                            overlay_slot: #Screw 0
                                                Padding:=margin{Left:=3.5,Top:=3.5,Right:=3.5,Bottom:=3.5}
                                                HorizontalAlignment:=horizontal_alignment.Left
                                                VerticalAlignment:=vertical_alignment.Top
                                                Widget:=texture_block{DefaultImage:=Images.Dialogue_Bolt, DefaultDesiredSize:=vector2{X:=5.0,Y:=5.0}}
                                            overlay_slot: #Screw 1
                                                Padding:=margin{Left:=3.5,Top:=3.5,Right:=3.5,Bottom:=3.5}
                                                HorizontalAlignment:=horizontal_alignment.Right
                                                VerticalAlignment:=vertical_alignment.Top
                                                Widget:=texture_block{DefaultImage:=Images.Dialogue_Bolt, DefaultDesiredSize:=vector2{X:=5.0,Y:=5.0}}
                                            overlay_slot: #Screw 2
                                                Padding:=margin{Left:=3.5,Top:=3.5,Right:=3.5,Bottom:=3.5}
                                                HorizontalAlignment:=horizontal_alignment.Left
                                                VerticalAlignment:=vertical_alignment.Bottom
                                                Widget:=texture_block{DefaultImage:=Images.Dialogue_Bolt, DefaultDesiredSize:=vector2{X:=5.0,Y:=5.0}}
                                            overlay_slot: #Screw 3
                                                Padding:=margin{Left:=3.5,Top:=3.5,Right:=3.5,Bottom:=3.5}
                                                HorizontalAlignment:=horizontal_alignment.Right
                                                VerticalAlignment:=vertical_alignment.Bottom
                                                Widget:=texture_block{DefaultImage:=Images.Dialogue_Bolt, DefaultDesiredSize:=vector2{X:=5.0,Y:=5.0}}
                                            overlay_slot:
                                                Padding:=margin{Left:=InventorySliderBarPadding,Top:=InventorySliderBarPadding,Right:=InventorySliderBarPadding,Bottom:=InventorySliderBarPadding}
                                                Widget:=PetPerk2_Text




                #Main overlay
                canvas_slot:
                    Anchors:=anchors{Minimum:=vector2{X:=0.0,Y:=0.5},Maximum:=vector2{X:=0.0,Y:=0.5}}
                    ZOrder:=0
                    SizeToContent:=true
                    Offsets:=margin{Left:=0.0,Top:=50.0,Right:=0.0,Bottom:=0.0} #PositionX, PositionY, SizeX, SizeY
                    Alignment:=vector2{X:=0.0,Y:=0.5}
                    Widget:=overlay:
                        Slots:=array:
                            overlay_slot:
                                HorizontalAlignment:=horizontal_alignment.Right
                                VerticalAlignment:=vertical_alignment.Center
                                Padding:=margin{Left:=5.0}
                                Widget:=stack_box:
                                    Orientation:=orientation.Vertical
                                        Slots:=array:
                                            stack_box_slot: #Sapphires
                                                Padding:=margin{Left:=ResourcersRowLeftPadding,Top:=ResourcersRowTopPadding,Right:=ResourcersRowRightPadding,Bottom:=ResourcersRowBottomPadding}
                                                HorizontalAlignment:=horizontal_alignment.Left
                                                VerticalAlignment:=vertical_alignment.Center
                                                Widget:=overlay:
                                                    Slots:=array:
                                                        overlay_slot:
                                                            HorizontalAlignment:=horizontal_alignment.Center
                                                            VerticalAlignment:=vertical_alignment.Center
                                                            Widget:=BackpackSapphires_Background
                                                        overlay_slot:
                                                            HorizontalAlignment:=horizontal_alignment.Center
                                                            VerticalAlignment:=vertical_alignment.Center
                                                            Widget:=BackpackSapphires_Border
                                                        overlay_slot:
                                                            HorizontalAlignment:=horizontal_alignment.Left
                                                            VerticalAlignment:=vertical_alignment.Center
                                                            Padding:=margin{Left:=InventoryIconPadding}
                                                            Widget:=BackpackSapphires_Icon
                                                        overlay_slot:
                                                            HorizontalAlignment:=horizontal_alignment.Left
                                                            VerticalAlignment:=vertical_alignment.Center
                                                            Padding:=margin{Left:=InventoryTextPadding}
                                                            Widget:=BackpackSapphires_Ammount

                                            stack_box_slot: #AggedSapphires
                                                Padding:=margin{Left:=ResourcersRowLeftPadding,Top:=ResourcersRowTopPadding,Right:=ResourcersRowRightPadding,Bottom:=ResourcersRowBottomPadding}
                                                HorizontalAlignment:=horizontal_alignment.Left
                                                VerticalAlignment:=vertical_alignment.Center
                                                Widget:=overlay:
                                                    Slots:=array:
                                                        overlay_slot:
                                                            HorizontalAlignment:=horizontal_alignment.Center
                                                            VerticalAlignment:=vertical_alignment.Center
                                                            Widget:=BackpackAgedSapphires_Background
                                                        overlay_slot:
                                                            HorizontalAlignment:=horizontal_alignment.Center
                                                            VerticalAlignment:=vertical_alignment.Center
                                                            Widget:=BackpackAgedSapphires_Border
                                                        overlay_slot:
                                                            HorizontalAlignment:=horizontal_alignment.Left
                                                            VerticalAlignment:=vertical_alignment.Center
                                                            Padding:=margin{Left:=InventoryIconPadding}
                                                            Widget:=BackpackAgedSapphires_Icon
                                                        overlay_slot:
                                                            HorizontalAlignment:=horizontal_alignment.Left
                                                            VerticalAlignment:=vertical_alignment.Center
                                                            Padding:=margin{Left:=InventoryTextPadding}
                                                            Widget:=BackpackAgedSapphires_Ammount
                                            stack_box_slot: #Water
                                                Padding:=margin{Left:=ResourcersRowLeftPadding,Top:=ResourcersRowTopPadding,Right:=ResourcersRowRightPadding,Bottom:=ResourcersRowBottomPadding}
                                                HorizontalAlignment:=horizontal_alignment.Fill
                                                VerticalAlignment:=vertical_alignment.Fill
                                                Widget:=overlay:
                                                    Slots:=array:
                                                        overlay_slot:
                                                            HorizontalAlignment:=horizontal_alignment.Left
                                                            VerticalAlignment:=vertical_alignment.Center
                                                            Padding:=margin{Left:=8.0}
                                                            Widget:=BackpackWater_EmptyPack
                                                        overlay_slot:
                                                            HorizontalAlignment:=horizontal_alignment.Left
                                                            VerticalAlignment:=vertical_alignment.Center
                                                            Padding:=margin{Left:=8.0}
                                                            Widget:=BackpackWater_Slider
                                                        overlay_slot:
                                                            HorizontalAlignment:=horizontal_alignment.Left
                                                            VerticalAlignment:=vertical_alignment.Center
                                                            Padding:=margin{Left:=20.0}
                                                            Widget:=BackpackWater_Icon
                                                        overlay_slot:
                                                            HorizontalAlignment:=horizontal_alignment.Center
                                                            VerticalAlignment:=vertical_alignment.Center
                                                            Widget:=BackpackWater_Border
                                                        overlay_slot:
                                                            HorizontalAlignment:=horizontal_alignment.Left
                                                            VerticalAlignment:=vertical_alignment.Center
                                                            Padding:=margin{Left:=70.0}
                                                            Widget:=BackpackWater_Ammount
                                            stack_box_slot: #Flour
                                                Padding:=margin{Left:=ResourcersRowLeftPadding,Top:=ResourcersRowTopPadding,Right:=ResourcersRowRightPadding,Bottom:=ResourcersRowBottomPadding}
                                                HorizontalAlignment:=horizontal_alignment.Fill
                                                VerticalAlignment:=vertical_alignment.Fill
                                                Widget:=overlay:
                                                    Slots:=array:
                                                        overlay_slot:
                                                            HorizontalAlignment:=horizontal_alignment.Left
                                                            VerticalAlignment:=vertical_alignment.Center
                                                            Padding:=margin{Left:=8.0}
                                                            Widget:=BackpackFlour_EmptyPack
                                                        overlay_slot:
                                                            HorizontalAlignment:=horizontal_alignment.Left
                                                            VerticalAlignment:=vertical_alignment.Center
                                                            Padding:=margin{Left:=8.0}
                                                            Widget:=BackpackFlour_Slider
                                                        overlay_slot:
                                                            HorizontalAlignment:=horizontal_alignment.Left
                                                            VerticalAlignment:=vertical_alignment.Center
                                                            Padding:=margin{Left:=20.0}
                                                            Widget:=BackpackFlour_Icon
                                                        overlay_slot:
                                                            HorizontalAlignment:=horizontal_alignment.Center
                                                            VerticalAlignment:=vertical_alignment.Center
                                                            Widget:=BackpackFlour_Border
                                                        overlay_slot:
                                                            HorizontalAlignment:=horizontal_alignment.Left
                                                            VerticalAlignment:=vertical_alignment.Center
                                                            Padding:=margin{Left:=70.0}
                                                            Widget:=BackpackFlour_Ammount
                                            stack_box_slot: #Cheese
                                                Padding:=margin{Left:=ResourcersRowLeftPadding,Top:=ResourcersRowTopPadding,Right:=ResourcersRowRightPadding,Bottom:=ResourcersRowBottomPadding}
                                                HorizontalAlignment:=horizontal_alignment.Fill
                                                VerticalAlignment:=vertical_alignment.Fill
                                                Widget:=overlay:
                                                    Slots:=array:
                                                        overlay_slot:
                                                            HorizontalAlignment:=horizontal_alignment.Left
                                                            VerticalAlignment:=vertical_alignment.Center
                                                            Padding:=margin{Left:=8.0}
                                                            Widget:=BackpackCheese_EmptyPack
                                                        overlay_slot:
                                                            HorizontalAlignment:=horizontal_alignment.Left
                                                            VerticalAlignment:=vertical_alignment.Center
                                                            Padding:=margin{Left:=8.0}
                                                            Widget:=BackpackCheese_Slider
                                                        overlay_slot:
                                                            HorizontalAlignment:=horizontal_alignment.Left
                                                            VerticalAlignment:=vertical_alignment.Center
                                                            Padding:=margin{Left:=20.0}
                                                            Widget:=BackpackCheese_Icon
                                                        overlay_slot:
                                                            HorizontalAlignment:=horizontal_alignment.Center
                                                            VerticalAlignment:=vertical_alignment.Center
                                                            Widget:=BackpackCheese_Border
                                                        overlay_slot:
                                                            HorizontalAlignment:=horizontal_alignment.Left
                                                            VerticalAlignment:=vertical_alignment.Center
                                                            Padding:=margin{Left:=70.0}
                                                            Widget:=BackpackCheese_Ammount
                                            stack_box_slot: #Pimentos
                                                Padding:=margin{Left:=ResourcersRowLeftPadding,Top:=ResourcersRowTopPadding,Right:=ResourcersRowRightPadding,Bottom:=ResourcersRowBottomPadding}
                                                HorizontalAlignment:=horizontal_alignment.Fill
                                                VerticalAlignment:=vertical_alignment.Fill
                                                Widget:=overlay:
                                                    Slots:=array:
                                                        overlay_slot:
                                                            HorizontalAlignment:=horizontal_alignment.Left
                                                            VerticalAlignment:=vertical_alignment.Center
                                                            Padding:=margin{Left:=8.0}
                                                            Widget:=BackpackPimentos_EmptyPack
                                                        overlay_slot:
                                                            HorizontalAlignment:=horizontal_alignment.Left
                                                            VerticalAlignment:=vertical_alignment.Center
                                                            Padding:=margin{Left:=8.0}
                                                            Widget:=BackpackPimentos_Slider
                                                        overlay_slot:
                                                            HorizontalAlignment:=horizontal_alignment.Left
                                                            VerticalAlignment:=vertical_alignment.Center
                                                            Padding:=margin{Left:=20.0}
                                                            Widget:=BackpackPimentos_Icon
                                                        overlay_slot:
                                                            HorizontalAlignment:=horizontal_alignment.Center
                                                            VerticalAlignment:=vertical_alignment.Center
                                                            Widget:=BackpackPimentos_Border
                                                        overlay_slot:
                                                            HorizontalAlignment:=horizontal_alignment.Left
                                                            VerticalAlignment:=vertical_alignment.Center
                                                            Padding:=margin{Left:=70.0}
                                                            Widget:=BackpackPimentos_Ammount
                                            stack_box_slot: #MutantMaterial
                                                Padding:=margin{Left:=ResourcersRowLeftPadding,Top:=ResourcersRowTopPadding,Right:=ResourcersRowRightPadding,Bottom:=ResourcersRowBottomPadding}
                                                HorizontalAlignment:=horizontal_alignment.Left
                                                VerticalAlignment:=vertical_alignment.Center
                                                Widget:=overlay:
                                                    Slots:=array:
                                                        overlay_slot:
                                                            HorizontalAlignment:=horizontal_alignment.Center
                                                            VerticalAlignment:=vertical_alignment.Center
                                                            Widget:=BackpackMutantMaterial_Background
                                                        overlay_slot:
                                                            HorizontalAlignment:=horizontal_alignment.Center
                                                            VerticalAlignment:=vertical_alignment.Center
                                                            Widget:=BackpackMutantMaterial_Border
                                                        overlay_slot:
                                                            HorizontalAlignment:=horizontal_alignment.Left
                                                            VerticalAlignment:=vertical_alignment.Center
                                                            Padding:=margin{Left:=InventoryIconPadding}
                                                            Widget:=BackpackMutantMaterial_Icon
                                                        overlay_slot:
                                                            HorizontalAlignment:=horizontal_alignment.Left
                                                            VerticalAlignment:=vertical_alignment.Center
                                                            Padding:=margin{Left:=InventoryTextPadding}
                                                            Widget:=BackpackMutantMaterial_Ammount
                                            stack_box_slot: #Coal
                                                Padding:=margin{Left:=ResourcersRowLeftPadding,Top:=ResourcersRowTopPadding,Right:=ResourcersRowRightPadding,Bottom:=ResourcersRowBottomPadding}
                                                HorizontalAlignment:=horizontal_alignment.Left
                                                VerticalAlignment:=vertical_alignment.Center
                                                Widget:=overlay:
                                                    Slots:=array:
                                                        overlay_slot:
                                                            HorizontalAlignment:=horizontal_alignment.Center
                                                            VerticalAlignment:=vertical_alignment.Center
                                                            Widget:=BackpackCoal_Background
                                                        overlay_slot:
                                                            HorizontalAlignment:=horizontal_alignment.Center
                                                            VerticalAlignment:=vertical_alignment.Center
                                                            Widget:=BackpackCoal_Border
                                                        overlay_slot:
                                                            HorizontalAlignment:=horizontal_alignment.Left
                                                            VerticalAlignment:=vertical_alignment.Center
                                                            Padding:=margin{Left:=InventoryIconPadding}
                                                            Widget:=BackpackCoal_Icon
                                                        overlay_slot:
                                                            HorizontalAlignment:=horizontal_alignment.Left
                                                            VerticalAlignment:=vertical_alignment.Center
                                                            Padding:=margin{Left:=InventoryTextPadding}
                                                            Widget:=BackpackCoal_Ammount



MakeMainInventoryUI<constructor>(
    InBakeryMachinesManager : bakery_machines_manager)<transacts> := main_inventory_ui :
        BakeryMachinesManager := InBakeryMachinesManager
                                                                                                                