using { /UnrealEngine.com/Temporary/UI }
using { /Verse.org/Simulation }
using { /Fortnite.com/UI }
using { /UnrealEngine.com/Temporary/SpatialMath }
using { /Verse.org/Colors/NamedColors }
using { /Verse.org/Colors }

# fill this file with your verse-script

CalculeAgedSapphires(LVArea : int, PickaxeTotalStrength : float, GameManager : game_manager) : float = 
    var TotalAgedSapphires : float = 0.0

    if(0<LVArea):
        Log("###",LOG_LEVEL_DEBUG4)
        TotalStrength:=PickaxeTotalStrength
        Log("TotalStrength: { GetOptimizedMagnitudeLetter(TotalStrength)}",LOG_LEVEL_DEBUG4)
        HardBreadLevelArea:=LVArea+1
        
        HardBreadHealth := GameManager.PickaxeCalculator.MainFormula_GenerateHardBreadHealth(HardBreadLevelArea,-1)
        Log("HardBreadLevelArea: {HardBreadLevelArea}, HardBreadHealth: {GetOptimizedMagnitudeLetter(HardBreadHealth)}",LOG_LEVEL_DEBUG4)
        ExpectedStrength:=HardBreadHealth/((MinNum_AVGHit+MaxNum_AVGHit)/2.0)

        AVGStrength:=(ExpectedStrength+TotalStrength)/2.0
        Log("AVGStrength: {GetOptimizedMagnitudeLetter(AVGStrength)}",LOG_LEVEL_DEBUG4)

        StrenghtExponential:=Log(10.0,AVGStrength)
        Log("StrenghtExponential: {GetOptimizedMagnitudeLetter(StrenghtExponential)}",LOG_LEVEL_DEBUG4)

        if(StrenghtExponential>=MinExponential_ForAgedSapphires*1.0):
            if(ExponentialDifference:=Floor[StrenghtExponential - (MinExponential_ForAgedSapphires*1.0)]):
                set TotalAgedSapphires = (ExponentialDifference*1.0) * AgedSapphiresPer_Exponential

    return TotalAgedSapphires

mastery_ui := class():

    BakeryManager : bakery_machines_manager  

    HLineHeight : float = 3.0
    HPadding : float = 5.0
    PaddingBetweenSection : float = 30.0
    KeepColor : color = Chartreuse
    ResetColor : color = Red

    var IsUIDisplayed : logic = false

    var MaybeDungeonCanvas : ?canvas = false

    CrownImage : texture_block = texture_block{DefaultImage:=Images.Cross_Icon, DefaultDesiredSize:=vector2{X:=128.0,Y:=128.0}}
    Mastery_TotalAgedSapphires_TextBlock : text_block = text_block{DefaultTextColor:=AgedSapphiresTextColor, DefaultShadowOffset:=ShadowOffset, DefaultShadowOpacity := DefaultTextShadowOpacity}
   

    var ConfirmationToMastery : logic = false
    ButtonMastery : button_quiet = button_quiet{}
    ButtonCancel : button_loud = button_loud{}

    InitUI() : void =
        ButtonCancel.SetText(BasicMessage("CANCEL"))
        ButtonMastery.SetText(BasicMessage("SELL BAKERY"))
        Canvas := MakeDungeonUICanvas()
        set MaybeDungeonCanvas = option{Canvas}
        InitSubscribes()
        UpdateUI()

    OnPlayerRemoved() : void =
        Remove_CanvasWithoutMessage()

    TryGetMastery(Message : widget_message) : void =
        if(Agent:=BakeryManager.MaybeAgentOwner?):
            AgedSapphiresAward:=CalculeAgedSapphires(BakeryManager.PlayerDataCache.LV_Area,BakeryManager.PickaxeTotalStrength, BakeryManager.GameManager)
            if(ConfirmationToMastery?):
                if(AgedSapphiresAward>0.0):
                    set ConfirmationToMastery = false
                    
                    NewNumMasteries := BakeryManager.PlayerDataCache.NumMasteries + 1
                    var NewPlayerDataCache : player_data_table = BakeryManager.GameManager.QuestsAndOrdersDatabase.AddAgentQuest_ProgressV2(Message.Player,BakeryManager.PlayerDataCache,1.0,MASTERY_QUEST_TYPE)(4)
                    set NewPlayerDataCache= BakeryManager.PlayersDataManager.SetMasteryReset(NewPlayerDataCache,AgedSapphiresAward)
                    set NewPlayerDataCache = BakeryManager.PlayersDataManager.SetNumMasteries(NewPlayerDataCache,NewNumMasteries)
                    #If the tutorial wa sno completed before, complete now
                    if(BakeryManager.PlayerDataCache.NumQuestsCompleted<Quest0_Index):
                        set NewPlayerDataCache = BakeryManager.PlayersDataManager.SetNumQuestsCompleted(NewPlayerDataCache,Quest0_Index+1)

                    set NewPlayerDataCache = BakeryManager.PlayersDataManager.CreateQuest_Delay(NewPlayerDataCache,Quest_DelayTime_Base)
                    set NewPlayerDataCache = BakeryManager.PlayersDataManager.CreateOrder_Delay(NewPlayerDataCache,Order_DelayTime_Base)
                    BakeryManager.SetNewPlayerDataCache(NewPlayerDataCache)
                    spawn:
                        BakeryManager.GameManager.VFXManager.PlayMasteryCeremony(Agent)
                    Remove_CanvasWithoutMessage()
                    BakeryManager.UpgradeResourcersContainersUI()
                    #BakeryManager.RefreshInventoryUI(BakeryManager.PlayerDataCache)
                    BakeryManager.RefreshAllUI()
                    BakeryManager.GameManager.CheckAreasToUnlock(BakeryManager.PlayerDataCache,Agent)
                    BakeryManager.GameManager.DisplayCharacterMessage(Agent,G_CHARACTER,"Bakery sold for {GetOptimizedMagnitudeLetter(AgedSapphiresAward)} Aged Sapphires!",?Sound:=BIGOBJETIVE_SOUND)
                    BakeryManager.Reset()
                    BakeryManager.DisplayResourcersGather(array{AgedSapphiresAward},array{AGEDSAPPHIRES},false)
                    spawn:
                        BakeryManager.LoadPlayerStats(?IsMastery:=true)
                    BakeryManager.GameManager.MiniGamesManager.RemoveScroll(BakeryManager.GameManager.GetAgentIndex(Agent))
                else:
                    Remove_CanvasWithoutMessage()
                    BakeryManager.GameManager.DisplayCharacterMessage(Agent,G_CHARACTER,"No point selling your Bakery for 0 Aged Sapphires!",?Sound:=ERROR_SOUND)

            else:
                ButtonMastery.SetText(BasicMessage("ARE YOU SURE?"))
                set ConfirmationToMastery = true

    UpdateUI() : void =
        TotalAgedSapphires:=CalculeAgedSapphires(BakeryManager.PlayerDataCache.LV_Area,BakeryManager.PickaxeTotalStrength, BakeryManager.GameManager)
        Mastery_TotalAgedSapphires_TextBlock.SetText(BasicMessage("x {GetOptimizedMagnitudeLetter(TotalAgedSapphires)}"))
        if(ConfirmationToMastery?):
            ButtonMastery.SetText(BasicMessage("ARE YOU SURE?"))
        else:
            ButtonMastery.SetText(BasicMessage("SELL BAKERY"))

    InitSubscribes() : void =
        ButtonMastery.OnClick().Subscribe(TryGetMastery)
        ButtonCancel.OnClick().Subscribe(Remove_Canvas)

    Remove_Canvas(Message : widget_message) : void =
        set IsUIDisplayed = false
        if:
            Agent:=BakeryManager.MaybeAgentOwner?
            Player:=player[Agent]
            PlayerUI:=GetPlayerUI[Player]
            Canvas := MaybeDungeonCanvas?
        then:
            PlayerUI.RemoveWidget(Canvas)
            RemoveDelay_OnRemove()


    Remove_CanvasWithoutMessage() : void =
        set IsUIDisplayed = false
        if:
            Agent:=BakeryManager.MaybeAgentOwner?
            Player:=player[Agent]
            PlayerUI:=GetPlayerUI[Player]
            Canvas := MaybeDungeonCanvas?
        then:
            PlayerUI.RemoveWidget(Canvas)
            RemoveDelay_OnRemove()

    RemoveDelay_OnAdd() : void =
        ButtonMastery.SetText(BasicMessage("SELL BAKERY"))
        ButtonCancel.SetText(BasicMessage("Cancel"))

    RemoveDelay_OnRemove() : void =
        ButtonCancel.SetText(BasicMessage("Cancel "))

    Display_Canvas() : void =
        if(not IsUIDisplayed?):
            set ConfirmationToMastery = false
            var Index : int = 0 
            if:
                Agent:=BakeryManager.MaybeAgentOwner?
                PlayerUI:=GetPlayerUI[player[Agent]]
                Canvas := MaybeDungeonCanvas?
            then:
                PlayerUI.AddWidget(Canvas,player_ui_slot{ZOrder:=0, InputMode := ui_input_mode.All})
                RemoveDelay_OnAdd()
        UpdateUI()

    MakeDungeonUICanvas(): canvas=
        Canvas := canvas:
            Slots:=array:
                #Main overlay
                canvas_slot:
                    Anchors:=anchors{Minimum:=vector2{X:=0.5,Y:=0.5},Maximum:=vector2{X:=0.5,Y:=0.5}}
                    ZOrder:=0
                    SizeToContent:=true
                    Offsets:=margin{Left:=0.0,Top:=0.0,Right:=0.0,Bottom:=0.0} #PositionX, PositionY, SizeX, SizeY
                    Alignment:=vector2{X:=0.5,Y:=0.5}
                    Widget:=texture_block{DefaultImage:=Images.BlackBackground,DefaultDesiredSize:=vector2{X:=BackgroundBlackUISizeX,Y:=BackgroundBlackUISizeY}}
                canvas_slot:
                    Anchors:=anchors{Minimum:=vector2{X:=0.5,Y:=0.5},Maximum:=vector2{X:=0.5,Y:=0.5}}
                    ZOrder:=0
                    SizeToContent:=true
                    Offsets:=margin{Left:=0.0,Top:=0.0,Right:=0.0,Bottom:=0.0} #PositionX, PositionY, SizeX, SizeY
                    Alignment:=vector2{X:=0.5,Y:=0.5}
                    Widget:=overlay:
                        Slots:=array:
                            overlay_slot:
                                HorizontalAlignment:=horizontal_alignment.Fill
                                VerticalAlignment:=vertical_alignment.Fill
                                Widget:=texture_block{DefaultImage:=Images.Img_MainBackground, DefaultDesiredSize:=vector2{X:=1300.0,Y:=1200.0}} 
                            overlay_slot:
                                HorizontalAlignment:=horizontal_alignment.Center
                                VerticalAlignment:=vertical_alignment.Center
                                #Padding:=margin{Top:=10.0,Bottom:=10.0,Right:=10.0,Left:=10.0}
                                Widget:=stack_box:  
                                    Orientation:=orientation.Vertical
                                        Slots:=array:
                                            stack_box_slot:
                                                Widget:=CrownImage 
                                            stack_box_slot:
                                                Widget:=text_block{DefaultText:= BasicMessage("SELL YOUR BAKERY AND GET AGED SAPPHIRES"), DefaultTextColor:=White, DefaultShadowOffset:=ShadowOffset, DefaultShadowOpacity := DefaultTextShadowOpacity}
                                            stack_box_slot:
                                                HorizontalAlignment:=horizontal_alignment.Fill
                                                Widget:=texture_block{DefaultImage:=Images.Img_HorizontalLine, DefaultDesiredSize:=vector2{X:=HLineHeight,Y:=HLineHeight}}
                                            stack_box_slot:
                                                Widget:=text_block{DefaultText:= BasicMessage("Get Aged Sapphires"), DefaultTextColor:=CurrentTextColor, DefaultShadowOffset:=ShadowOffset, DefaultShadowOpacity := DefaultTextShadowOpacity}
                                            stack_box_slot: #Get Aged Sapphires
                                                VerticalAlignment:=vertical_alignment.Fill
                                                HorizontalAlignment:=horizontal_alignment.Fill 
                                                Widget:=overlay:
                                                    Slots:=array:
                                                        overlay_slot:
                                                            HorizontalAlignment:=horizontal_alignment.Fill
                                                            VerticalAlignment:=vertical_alignment.Fill
                                                            Widget:=texture_block{DefaultImage:=Images.Img_LineBackground, DefaultDesiredSize:=vector2{X:=HLineHeight,Y:=HLineHeight}}
                                                        overlay_slot:
                                                            HorizontalAlignment:=horizontal_alignment.Center
                                                            VerticalAlignment:=vertical_alignment.Center
                                                            Widget:=stack_box:  
                                                                Orientation:=orientation.Horizontal
                                                                Slots:=array:
                                                                    stack_box_slot:
                                                                        HorizontalAlignment:=horizontal_alignment.Center
                                                                        VerticalAlignment:=vertical_alignment.Center
                                                                        Widget:=texture_block{DefaultImage:=Images.AgedSaphire_Icon, DefaultDesiredSize:=vector2{X:=SizeSmallIconX,Y:=SizeSmallIconY}}
                                                                    stack_box_slot:
                                                                        HorizontalAlignment:=horizontal_alignment.Center
                                                                        VerticalAlignment:=vertical_alignment.Center
                                                                        Padding:=margin{Left:=HPadding}
                                                                        Widget:=Mastery_TotalAgedSapphires_TextBlock
                                                                    stack_box_slot:
                                                                        HorizontalAlignment:=horizontal_alignment.Center
                                                                        VerticalAlignment:=vertical_alignment.Center
                                                                        Padding:=margin{Left:=HPadding}
                                                                        Widget:=text_block{DefaultText:= BasicMessage("(Applies Pickaxe Damage & Area LV unlocked)"), DefaultTextColor:=White, DefaultShadowOffset:=ShadowOffset, DefaultShadowOpacity := DefaultTextShadowOpacity}
                                            stack_box_slot:
                                                HorizontalAlignment:=horizontal_alignment.Fill
                                                Widget:=texture_block{DefaultImage:=Images.Img_HorizontalLine, DefaultDesiredSize:=vector2{X:=HLineHeight,Y:=HLineHeight}}                    
                                            stack_box_slot:
                                                Widget:=text_block{DefaultText:= BasicMessage("Keep"), DefaultTextColor:=CurrentTextColor, DefaultShadowOffset:=ShadowOffset, DefaultShadowOpacity := DefaultTextShadowOpacity}
                                            stack_box_slot: 
                                                VerticalAlignment:=vertical_alignment.Fill
                                                HorizontalAlignment:=horizontal_alignment.Fill 
                                                Widget:=overlay:
                                                    Slots:=array:
                                                        overlay_slot:
                                                            HorizontalAlignment:=horizontal_alignment.Fill
                                                            VerticalAlignment:=vertical_alignment.Fill
                                                            Widget:=texture_block{DefaultImage:=Images.Img_LineBackground, DefaultDesiredSize:=vector2{X:=HLineHeight,Y:=HLineHeight}}
                                                        overlay_slot:
                                                            HorizontalAlignment:=horizontal_alignment.Left
                                                            VerticalAlignment:=vertical_alignment.Center
                                                            Widget:=text_block{DefaultText:= BasicMessage("-Any upgrade made with Aged Sapphires, Coal or Mutant Material"), DefaultTextColor:=KeepColor, DefaultShadowOffset:=ShadowOffset, DefaultShadowOpacity := DefaultTextShadowOpacity}
                                            stack_box_slot: 
                                                VerticalAlignment:=vertical_alignment.Fill
                                                HorizontalAlignment:=horizontal_alignment.Fill 
                                                Widget:=overlay:
                                                    Slots:=array:
                                                        overlay_slot:
                                                            HorizontalAlignment:=horizontal_alignment.Fill
                                                            VerticalAlignment:=vertical_alignment.Fill
                                                            Widget:=texture_block{DefaultImage:=Images.Img_LineBackground, DefaultDesiredSize:=vector2{X:=HLineHeight,Y:=HLineHeight}}
                                                        overlay_slot:
                                                            HorizontalAlignment:=horizontal_alignment.Left
                                                            VerticalAlignment:=vertical_alignment.Center
                                                            Widget:=text_block{DefaultText:= BasicMessage("-Fame, Aged Sapphires, Coal and Mutant Material"), DefaultTextColor:=KeepColor, DefaultShadowOffset:=ShadowOffset, DefaultShadowOpacity := DefaultTextShadowOpacity}
                                            stack_box_slot: 
                                                VerticalAlignment:=vertical_alignment.Fill
                                                HorizontalAlignment:=horizontal_alignment.Fill 
                                                Widget:=overlay:
                                                    Slots:=array:
                                                        overlay_slot:
                                                            HorizontalAlignment:=horizontal_alignment.Fill
                                                            VerticalAlignment:=vertical_alignment.Fill
                                                            Widget:=texture_block{DefaultImage:=Images.Img_LineBackground, DefaultDesiredSize:=vector2{X:=HLineHeight,Y:=HLineHeight}}
                                                        overlay_slot:
                                                            HorizontalAlignment:=horizontal_alignment.Left
                                                            VerticalAlignment:=vertical_alignment.Center
                                                            Widget:=text_block{DefaultText:= BasicMessage("-Secret Formulas and Sourdough Challenges found"), DefaultTextColor:=KeepColor, DefaultShadowOffset:=ShadowOffset, DefaultShadowOpacity := DefaultTextShadowOpacity}
                                            stack_box_slot: 
                                                VerticalAlignment:=vertical_alignment.Fill
                                                HorizontalAlignment:=horizontal_alignment.Fill 
                                                Widget:=overlay:
                                                    Slots:=array:
                                                        overlay_slot:
                                                            HorizontalAlignment:=horizontal_alignment.Fill
                                                            VerticalAlignment:=vertical_alignment.Fill
                                                            Widget:=texture_block{DefaultImage:=Images.Img_LineBackground, DefaultDesiredSize:=vector2{X:=HLineHeight,Y:=HLineHeight}}
                                                        overlay_slot:
                                                            HorizontalAlignment:=horizontal_alignment.Left
                                                            VerticalAlignment:=vertical_alignment.Center
                                                            Widget:=text_block{DefaultText:= BasicMessage("-Sourdoughs and Sourdough Competition Victories"), DefaultTextColor:=KeepColor, DefaultShadowOffset:=ShadowOffset, DefaultShadowOpacity := DefaultTextShadowOpacity}
                                            stack_box_slot:
                                                HorizontalAlignment:=horizontal_alignment.Fill
                                                Widget:=texture_block{DefaultImage:=Images.Img_HorizontalLine, DefaultDesiredSize:=vector2{X:=HLineHeight,Y:=HLineHeight}}                    
                                            stack_box_slot:
                                                Widget:=text_block{DefaultText:= BasicMessage("Reset"), DefaultTextColor:=CurrentTextColor, DefaultShadowOffset:=ShadowOffset, DefaultShadowOpacity := DefaultTextShadowOpacity}
                                            stack_box_slot: 
                                                VerticalAlignment:=vertical_alignment.Fill
                                                HorizontalAlignment:=horizontal_alignment.Fill 
                                                Widget:=overlay:
                                                    Slots:=array:
                                                        overlay_slot:
                                                            HorizontalAlignment:=horizontal_alignment.Fill
                                                            VerticalAlignment:=vertical_alignment.Fill
                                                            Widget:=texture_block{DefaultImage:=Images.Img_LineBackground, DefaultDesiredSize:=vector2{X:=HLineHeight,Y:=HLineHeight}}
                                                        overlay_slot:
                                                            HorizontalAlignment:=horizontal_alignment.Left
                                                            VerticalAlignment:=vertical_alignment.Center
                                                            Widget:=text_block{DefaultText:= BasicMessage("-Any upgrade made with Sapphires"), DefaultTextColor:=ResetColor, DefaultShadowOffset:=ShadowOffset, DefaultShadowOpacity := DefaultTextShadowOpacity}
                                            stack_box_slot: 
                                                VerticalAlignment:=vertical_alignment.Fill
                                                HorizontalAlignment:=horizontal_alignment.Fill 
                                                Widget:=overlay:
                                                    Slots:=array:
                                                        overlay_slot:
                                                            HorizontalAlignment:=horizontal_alignment.Fill
                                                            VerticalAlignment:=vertical_alignment.Fill
                                                            Widget:=texture_block{DefaultImage:=Images.Img_LineBackground, DefaultDesiredSize:=vector2{X:=HLineHeight,Y:=HLineHeight}}
                                                        overlay_slot:
                                                            HorizontalAlignment:=horizontal_alignment.Left
                                                            VerticalAlignment:=vertical_alignment.Center
                                                            Widget:=text_block{DefaultText:= BasicMessage("-Any upgrade made Water, Flour, Cheese and Pimentos"), DefaultTextColor:=ResetColor, DefaultShadowOffset:=ShadowOffset, DefaultShadowOpacity := DefaultTextShadowOpacity}
                                            stack_box_slot: 
                                                VerticalAlignment:=vertical_alignment.Fill
                                                HorizontalAlignment:=horizontal_alignment.Fill 
                                                Widget:=overlay:
                                                    Slots:=array:
                                                        overlay_slot:
                                                            HorizontalAlignment:=horizontal_alignment.Fill
                                                            VerticalAlignment:=vertical_alignment.Fill
                                                            Widget:=texture_block{DefaultImage:=Images.Img_LineBackground, DefaultDesiredSize:=vector2{X:=HLineHeight,Y:=HLineHeight}}
                                                        overlay_slot:
                                                            HorizontalAlignment:=horizontal_alignment.Left
                                                            VerticalAlignment:=vertical_alignment.Center
                                                            Widget:=text_block{DefaultText:= BasicMessage("-All Backpack Resourcers, Tokens and Super Energy"), DefaultTextColor:=ResetColor, DefaultShadowOffset:=ShadowOffset, DefaultShadowOpacity := DefaultTextShadowOpacity}
                                            stack_box_slot: 
                                                VerticalAlignment:=vertical_alignment.Fill
                                                HorizontalAlignment:=horizontal_alignment.Fill 
                                                Widget:=overlay:
                                                    Slots:=array:
                                                        overlay_slot:
                                                            HorizontalAlignment:=horizontal_alignment.Fill
                                                            VerticalAlignment:=vertical_alignment.Fill
                                                            Widget:=texture_block{DefaultImage:=Images.Img_LineBackground, DefaultDesiredSize:=vector2{X:=HLineHeight,Y:=HLineHeight}}
                                                        overlay_slot:
                                                            HorizontalAlignment:=horizontal_alignment.Left
                                                            VerticalAlignment:=vertical_alignment.Center
                                                            Widget:=text_block{DefaultText:= BasicMessage("-Current Quest and Order"), DefaultTextColor:=ResetColor, DefaultShadowOffset:=ShadowOffset, DefaultShadowOpacity := DefaultTextShadowOpacity}
                                            stack_box_slot: 
                                                Padding:=margin{Top:=20.0}
                                                VerticalAlignment:=vertical_alignment.Fill
                                                HorizontalAlignment:=horizontal_alignment.Fill 
                                                Widget:=overlay:
                                                    Slots:=array:
                                                        overlay_slot:
                                                            HorizontalAlignment:=horizontal_alignment.Fill
                                                            VerticalAlignment:=vertical_alignment.Fill
                                                            Widget:=texture_block{DefaultImage:=Images.EmptyTexture, DefaultDesiredSize:=vector2{X:=960.0,Y:=SizeBoxSaphiresY}}
                                                        overlay_slot:
                                                            HorizontalAlignment:=horizontal_alignment.Fill
                                                            VerticalAlignment:=vertical_alignment.Fill
                                                            Widget:=ButtonMastery
                                            stack_box_slot: 
                                                Padding:=margin{Top:=20.0}
                                                VerticalAlignment:=vertical_alignment.Fill
                                                HorizontalAlignment:=horizontal_alignment.Fill 
                                                Widget:=overlay:
                                                    Slots:=array:
                                                        overlay_slot:
                                                            HorizontalAlignment:=horizontal_alignment.Fill
                                                            VerticalAlignment:=vertical_alignment.Fill
                                                            Widget:=texture_block{DefaultImage:=Images.EmptyTexture, DefaultDesiredSize:=vector2{X:=960.0,Y:=SizeBoxSaphiresY}}
                                                        overlay_slot:
                                                            HorizontalAlignment:=horizontal_alignment.Fill
                                                            VerticalAlignment:=vertical_alignment.Fill
                                                            Widget:=ButtonCancel
                            
MakeMasteryUI<constructor>(InBakeryManager : bakery_machines_manager
    )<transacts> := mastery_ui :
        BakeryManager := InBakeryManager
                                                       

                            
           


