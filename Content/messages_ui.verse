using { /UnrealEngine.com/Temporary/UI }
using { /UnrealEngine.com/Temporary/SpatialMath }
using { /Verse.org/Simulation }
using { /Fortnite.com/Devices }
using { /Fortnite.com/UI }
using { /Verse.org/Colors/NamedColors }
using { /Verse.org/Colors }

# See https://dev.epicgames.com/documentation/en-us/uefn/create-your-own-device-in-verse for how to create a verse device.

# A Verse-authored creative device that can be placed in a level
messages_ui := class():

    #Class settings
    var MaybeAgent : ?agent = false

    BakeryMachinesManager : bakery_machines_manager

    RColumnSizeX : float = 200.0
    RColumnSizeY : float = 50.0

    IconPadding:margin=margin{Left:=5.0,Top:=5.0,Right:=5.0,Bottom:=5.0}
    TextPadding:margin=margin{Left:=60.0,Top:=5.0,Right:=5.0,Bottom:=5.0}

    BackgroundLayer1_TextureBlock : texture_block = texture_block{DefaultImage:=Images.Img_LineBackground}
    BackgroundLayer2_TextureBlock : texture_block = texture_block{DefaultImage:=Images.Img_LineBackground}

    CharacterImage_TextureBlock : texture_block = texture_block{DefaultDesiredSize:=vector2{X:=100.0,Y:=100.0}, DefaultImage:=Images.Mike_Icon}

    Empty_TextureBlock : texture_block = texture_block{DefaultDesiredSize:=vector2{X:=500.0,Y:=200.0}, DefaultImage:=Images.EmptyTexture}

    Title_TextBlock : text_block = text_block{DefaultTextColor:=Red, DefaultShadowOffset:=ShadowOffset, DefaultShadowOpacity := DefaultTextShadowOpacity}
    Description_TextBlock : text_block = text_block{DefaultTextColor:=White, DefaultShadowOffset:=ShadowOffset, DefaultShadowOpacity := DefaultTextShadowOpacity}
    Progress_TextBlock : text_block = text_block{DefaultTextColor:=CurrentTextColor, DefaultShadowOffset:=ShadowOffset, DefaultShadowOpacity := DefaultTextShadowOpacity}

    MainOverlay : overlay =  overlay{}

    var MaybeCanvas : ?canvas = false
    var UIIsDisplayed : logic = false

    var TimeElapsed : int = 0

    InitUI() : void =
        UpdateUI()
        Canvas := MakeCanvas()
        set MaybeCanvas = option{Canvas}
        #RemoveDelay_OnRemove()

    SetAgent(InAgent : agent) : void =
        set MaybeAgent = option{InAgent}

    UpdateUI() : void =

        Title_TextBlock.SetText(BasicMessage("Do things"))
        Description_TextBlock.SetText(BasicMessage("You need to do things"))
        Progress_TextBlock.SetText(BasicMessage("1/25"))

    Remove_Canvas(Message : widget_message) : void =
        set UIIsDisplayed = false
        #Print("Trying to remove UI")
        if:
            Agent := MaybeAgent?
            Player:=player[Agent]
            PlayerUI:=GetPlayerUI[Player]
            Canvas := MaybeCanvas?
        then:
            #Print("Removing UI")
            PlayerUI.RemoveWidget(Canvas)
            #RemoveDelay_OnRemove()
            #SetValues_Canvas()

    Remove_CanvasWithoutMessage() : void =
        set UIIsDisplayed = false
        #Print("Trying to remove UI")
        if:
            Agent := MaybeAgent?
            Player:=player[Agent]
            PlayerUI:=GetPlayerUI[Player]
            Canvas := MaybeCanvas?
        then:
            #Print("Removing UI")
            PlayerUI.RemoveWidget(Canvas)
            #RemoveDelay_OnRemove()
            #SetValues_Canvas()

            
    Display_Canvas() : void =
        set TimeElapsed = 0
        set UIIsDisplayed = true
        spawn: 
            StartTimeElapsed()
        if:
            Agent := MaybeAgent?
            PlayerUI:=GetPlayerUI[player[Agent]]
            Canvas := MaybeCanvas?
        then:
            #Print("Assigning created UI before")
            PlayerUI.AddWidget(Canvas,player_ui_slot{ZOrder:=1})
            #SetValues_Canvas()
            #RemoveDelay_OnAdd()   
        #UpdateUI()

    StartTimeElapsed()<suspends> : void =
        loop:
            #Print("IN THE LOOP TimeElapsed: {TimeElapsed}")
            if(TimeElapsed > 20):
                #Print("Closing the UI because 20 seconds has pased")
                Remove_CanvasWithoutMessage()
                break
            if(UIIsDisplayed = false):
                #Print("Breaking Time Elapsed loop")
                break
            Sleep(1.0)
            set TimeElapsed = TimeElapsed + 1

            <#
    ChangeValues(Message : widget_message) : void =
        #Print("hey!")
        #CurrentBuyOption_textblock.SetText(BasicMessage("Buy x1"))
        #CurrentBuyOption_textblock.SetVisibility(widget_visibility.Hidden)

    RemoveDelay_OnAdd() : void =
        #Print("hey!")
        #CurrentBuyOption_textblock.SetText(BasicMessage("Buy x1"))
        #CurrentBuyOption_textblock.SetVisibility(widget_visibility.Hidden)

    RemoveDelay_OnRemove() : void =
        #Print("hey!")
        #CurrentBuyOption_textblock.SetText(BasicMessage("Buy x2"))
        #CurrentBuyOption_textblock.SetVisibility(widget_visibility.Hidden)

    SetValues_Canvas() : void =
        #Print("hey!")
        #BuyOptionx1_buttonquiet.SetText(BasicMessage("x1"))#>

    MakeCanvas(): canvas=
        Canvas := canvas:
            Slots:=array:
                #Main overlay
                canvas_slot:
                    ZOrder:=0
                    Anchors:=anchors{Minimum:=vector2{X:=1.0,Y:=0.5},Maximum:=vector2{X:=1.0,Y:=0.5}}
                    Offsets:=margin{Left:=0.0,Top:=0.0,Right:=0.0,Bottom:=0.0} #PositionX, PositionY, SizeX, SizeY
                    Alignment:=vector2{X:=1.0,Y:=0.5}
                    SizeToContent:=true
                    Widget:=overlay:
                        Slots:=array:
                            overlay_slot:
                                Padding:=margin{Left:=2.0,Top:=2.0,Right:=2.0,Bottom:=2.0}
                                HorizontalAlignment:=horizontal_alignment.Fill
                                VerticalAlignment:=vertical_alignment.Fill
                                Widget:=BackgroundLayer1_TextureBlock
                            overlay_slot:
                                HorizontalAlignment:=horizontal_alignment.Fill
                                VerticalAlignment:=vertical_alignment.Fill
                                Widget:=BackgroundLayer2_TextureBlock
                            overlay_slot:
                                Padding:=margin{Left:=50.0,Top:=-100.0,Right:=50.0,Bottom:=0.0}
                                HorizontalAlignment:=horizontal_alignment.Left
                                VerticalAlignment:=vertical_alignment.Top
                                Widget:=CharacterImage_TextureBlock
                            overlay_slot:
                                HorizontalAlignment:=horizontal_alignment.Fill
                                VerticalAlignment:=vertical_alignment.Fill
                                    Widget:=overlay:
                                        Slots:=array:
                                            overlay_slot: 
                                                HorizontalAlignment:=horizontal_alignment.Center
                                                VerticalAlignment:=vertical_alignment.Top
                                                Padding:=margin{Left:=0.0,Top:=10.0,Right:=0.0,Bottom:=0.0}
                                                Widget:=Title_TextBlock
                                            overlay_slot: 
                                                Padding:=margin{Left:=5.0,Top:=5.0,Right:=5.0,Bottom:=5.0}
                                                HorizontalAlignment:=horizontal_alignment.Fill
                                                VerticalAlignment:=vertical_alignment.Center
                                                Widget:=Description_TextBlock
                                            overlay_slot: 
                                                HorizontalAlignment:=horizontal_alignment.Center
                                                VerticalAlignment:=vertical_alignment.Bottom
                                                Padding:=margin{Left:=0.0,Top:=0.0,Right:=0.0,Bottom:=15.0}
                                                Widget:=Progress_TextBlock
                                            overlay_slot: 
                                                Widget:=Empty_TextureBlock                                          
MakeMessagesUI<constructor>(
    InBakeryMachinesManager : bakery_machines_manager)<transacts> := messages_ui :
        BakeryMachinesManager := InBakeryMachinesManager
                                                                                                               