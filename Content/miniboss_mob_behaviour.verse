using { /Fortnite.com/AI }
using { /Fortnite.com/Characters }
using { /Verse.org }
using { /Verse.org/Colors }
using { /Verse.org/Random }
using { /Verse.org/Simulation }
using { /UnrealEngine.com/Temporary/Diagnostics }
using { /UnrealEngine.com/Temporary/SpatialMath }
using { /Verse.org/Assets }
using { /Fortnite.com/Animation/PlayAnimation }
using { /Fortnite.com/Devices }
using { /Fortnite.com/Game }

# Getting started: 
# https://www.epicgames.com/fortnite/en-US/creative/docs/uefn/Verse/onboarding-guide-to-programming-with-verse-in-unreal-editor-for-fortnite
  
# Create a dedicated debug channel to draw to for this behavior
miniboss_mob_behaviour_debug_draw := class(debug_draw_channel) {}

# This module contains functions which return messages used for helpful debugging information that are
# localized and paired with the NPC ID.
miniboss_mob_behaviour_message_module:= module:

    SettingsCategory<public><localizes>:message = "Settings"

    OnBeginMessage<public><localizes>(Agent:agent):message =
        "NPC Agent = {Agent}: OnBegin triggered let's get started."
    OnEndMessage<public><localizes>(Agent:agent):message =
        "NPC Agent = {Agent}: OnEnd triggered let's cleanup."
    OnNavigateBeginMessage<public><localizes>(Agent:agent, X:float, Y:float, Z:float):message =
        "NPC Agent = {Agent}: Is moving to [{X},{Y},{Z}]"
    OnNavigateErrorMessage<public><localizes>(Agent:agent, X:float, Y:float, Z:float):message =
        "NPC Agent = {Agent}: Hit error moving to [{X},{Y},{Z}] from [{CurrentX},{CurrentY},{CurrentZ}], please refer to Island Setting's Navigation debug, please refer to Island Setting's Navigation debug"

# A Verse-authored NPC Behavior that can be used within an NPC Character Definition or a NPC Spawner device's NPC Behavior Script Override.
miniboss_mob_behaviour := class(npc_behavior):

    THIS_MOBTYPE : int = MINIBOSS_MOB

    var NPCSpawnPoint : transform = transform{}

    var NumNavigationErrors : int = 0

    var MaybeMobBehaviourHelper : ?mob_behavior_helper = false

    var NumSpecialActions : int = 0
    var LastSpecialAction : int = 0
    var Main_DamageToPlayer : float = 5.0
    var Secondary_DamageToPlayer : float = 5.0

    MiniBoss_CheckArena : float = 1500.0

    NPCDamage : float = 128.0
    NPCRange : float = 128.0
    #var NPCNavigationTarget : navigation_target
    var MaybeCharacter : ?fort_character = false
    # How long to wait in seconds after the NPC navigates to a point before moving on.
    @editable_number(float):
        Categories:=array{ground_mob_behaviour_message_module.SettingsCategory},
        MinValue:=option{0.5},
        MaxValue:=option{5.0}
    MoveToWaitDuration:float = 0.5

    # The negative min and absolute max x & y coordinate offset in centimeters to tell the NPC to move to
    @editable_number(float):
        Categories:=array{ground_mob_behaviour_message_module.SettingsCategory},
        MinValue:=option{0.0}
    DistanceFromSpawnPtToMove:float = 150.0

    # Whether to draw debug to the NPC channel when Verse Debug Draw is enabled in Island Settings.
    @editable:
        Categories:=array{ground_mob_behaviour_message_module.SettingsCategory}
    ShowAIDebug:logic = true

    # How long in seconds to render the debug draw location and print text. 
    # It is recommended to keep this in sync with MoveToWaitDuration otherwise the print will not be shown if a previous message is displayed.
    @editable_number(float):
        Categories:=array{ground_mob_behaviour_message_module.SettingsCategory},
        MinValue:=option{0.5}
    AIDebugDrawTime:float = 5.0

    # How long in seconds to render the look at arrow's debug draw.
    @editable_number(float):
        Categories:=array{ground_mob_behaviour_message_module.SettingsCategory},
        MinValue:=option{0.5}
    LookAtDebugDrawDuration:float = 0.5

    # Debug Draw instance.
    DebugDrawNPC:debug_draw = debug_draw{Channel := ground_mob_behaviour_debug_draw}

    # Used for specifying a point offset from the NPC pelvis to the head to draw the look at arrow from.
    VerticalOffsetToNPCHead<private>:float = 55.0

    <#
#DUPLICATED CODE
    ##############
    #DUPLICATED VARIABLES
    DruidMobAttack  : animation_sequence = EnemyMobs.MutantMiniMob_Druid_Anim_AttackBite
    GroundMobAttack  : animation_sequence = EnemyMobs.MutantMiniMob_Druid_Anim_AttackBite
    MiniBossMobAttack  : animation_sequence = EnemyMobs.B_1_Anim_Attack
    PlantMobAttack  : animation_sequence = EnemyMobs.Plant_Anim_Attack_001

    SmallSpitRange : float = 242.0
    LongSpitRange : float = 362.0
    MiniBossSpitRange : float = 362.0

    SmallSpitDuration : float = 1.0
    LongSpitDuration : float = 1.0
    MiniBossSpitDuration : float = 0.5

    ShortAttack_PreDelay : float = 1.125
    ShortAttack_PostDelay : float = 0.25

    LongAttack_PreDelay : float = 1.224
    LongAttack_PostDelay : float = 0.375

    MiniBossAttack_PreDelay : float = 0.945
    MiniBossAttack_PostDelay : float = 0.2

    #MOBS
    Mob_EnemyRangeDetection : float = 1500.0
    Mob_NavigationErrorWait : float = 2.0

    #DamageZone
    DamageZonePreDamage_Delay : float = 2.0
    DamageZonePostDamage_Delay : float = 3.5
    DamageZoneRadius : float = 335.0

    #MiniBoss
    MiniBoss_MaxSpecialActions : int = 3
    MiniBoss_WaitBetweenSpecialActions : int = 15

    MiniBoss_Max_Movement  : float = 512.0
    MiniBossBite_Damage : float = 20.0
    MiniBoss_AreaDamage : float = 10.0
    MiniBoss_MoveToWait_Duration : float = 0.25
    MiniBoss_ActionToWait_Duration : float = 0.5
    MiniBoss_MinionsSpawnToWait_Duration : float = 2.0

    MiniBossAttack_Range : float = 256.0

    <#
    #Flying mob
    FlyingMob_MoveToWait_Duration : float = 0.15
    FlyingMob_Damage : float = 5.0
    FlyingMobAttack_Range : float = 128.0
    FlyingMobAttack_PreDelay : float = 0.5
    FlyingMobAttack_PostDelay : float = 1.0#>

    #GroundMobBite_Damage : float = 5.0

    #Ground mob
    GroundMob_Max_Movement : float = 1024.0
    GroundMob_MoveToWait_Duration : float = 0.2
    GroundMob_Damage : float = 15.0
    GroundMobAttack_Range : float = 188.0
    #GroundMobAttack_PreDelay : float = 1.6
    #GroundMobAttack_PostDelay : float = 0.5

    #Druid mob
    Druid_WaitBetweenSpecialActions : int = 10

    #HealZone
    HealZonePreDamage_Delay : float = 2.0
    HealZonePostDamage_Delay : float = 3.5
    MobHealZoneRadius : float = 335.0

    DruidMob_Max_Movement : float = 1024.0
    DruidMob_MoveToWait_Duration : float = 0.2
    DruidMob_Damage : float = 10.0
    DruidMobAttack_Range : float = 188.0
    DruidMobAttack_PreDelay : float = 1.6
    DruidMobAttack_PostDelay : float = 0.7

    #Plant Mob
    
    PreSpitWait : float = 0.5
    PostSpitWait : float = 0.5

    PlantMob_MoveToWait_Duration : float = 0.1
    PlantMob_Damage : float = 25.0
    PlantMobAttack_Range : float = 318.0
    ##############
    #DUPLICATED MAKE MOVE
    MakeMovementDecision(NPCAgent : agent,NPCCharacter : fort_character, MobType : int) : tuple(?agent,?vector3,?vector3,logic,float,float) =
        var Mob_MoveToWait_Duration : float = 1.0
        var MaybeAgentObjetive : ?agent = false
        var MaybePrimary_TargetPosition : ?vector3 = false
        var MaybeSecondary_TargetPosition : ?vector3 = false
        
        var IsClippingDetected : logic = false
        
        var ReachRadius : float = 64.0
        
        if(MobType=MINIBOSS_MOB):
            set ReachRadius = MiniBossAttack_Range
            if(NewTargetCharacter:=MobBehaviourHelper.FindNearestTargetFromPos[NPCCharacter,Mob_EnemyRangeDetection], TargetAgent := NewTargetCharacter.GetAgent[]):
                set MaybeAgentObjetive = option{TargetAgent}
            Primary_TargetPosition := MobBehaviourHelper.FindRandomNavigationPoint(NPCCharacter.GetTransform().Translation,MiniBoss_Max_Movement,NPCSpawnPoint.Translation)
            set MaybePrimary_TargetPosition = option{Primary_TargetPosition}
            set MaybeSecondary_TargetPosition = option{Primary_TargetPosition}
            if(10.0<Abs(Primary_TargetPosition.Z-NPCSpawnPoint.Translation.Z)):
                set IsClippingDetected = true
                #Print("Detected clipping error")
        else if(MobType=PLANT_MOB):
            set Mob_MoveToWait_Duration = PlantMob_MoveToWait_Duration
            if(NewTargetCharacter:=MobBehaviourHelper.FindNearestTargetFromPos[NPCCharacter,Mob_EnemyRangeDetection], TargetAgent := NewTargetCharacter.GetAgent[]):
                #Find enemy
                set MaybeAgentObjetive = option{TargetAgent}
                #The plant does not move
                set MaybePrimary_TargetPosition = option{NPCSpawnPoint.Translation}
                set MaybeSecondary_TargetPosition = option{NPCSpawnPoint.Translation}

        else if(MobType=DRUID_MOB):
            set ReachRadius = DruidMobAttack_Range-56.0
            if(NewTargetCharacter:=MobBehaviourHelper.FindNearestTargetFromPos[NPCCharacter,Mob_EnemyRangeDetection], TargetAgent := NewTargetCharacter.GetAgent[]):
                #Find enemy
                set MaybeAgentObjetive = option{TargetAgent}
                set MaybePrimary_TargetPosition = option{NewTargetCharacter.GetTransform().Translation}
                #Secondary target
                Secondary_TargetPosition := MobBehaviourHelper.FindRandomNavigationPoint(NPCCharacter.GetTransform().Translation, GroundMob_Max_Movement,NPCSpawnPoint.Translation)
                set MaybeSecondary_TargetPosition = option{Secondary_TargetPosition}
            else:
                #Creation two posible positions to move
                Primary_TargetPosition := MobBehaviourHelper.FindRandomNavigationPoint(NPCCharacter.GetTransform().Translation, GroundMob_Max_Movement,NPCSpawnPoint.Translation)
                set MaybePrimary_TargetPosition = option{Primary_TargetPosition}
                Secondary_TargetPosition := MobBehaviourHelper.FindRandomNavigationPoint(NPCCharacter.GetTransform().Translation, GroundMob_Max_Movement,NPCSpawnPoint.Translation)
                set MaybeSecondary_TargetPosition = option{Secondary_TargetPosition}

        else if(MobType=GROUND_MOB):
            set ReachRadius = GroundMobAttack_Range-56.0
            if(NewTargetCharacter:=MobBehaviourHelper.FindNearestTargetFromPos[NPCCharacter,Mob_EnemyRangeDetection], TargetAgent := NewTargetCharacter.GetAgent[]):
                #Find enemy
                set MaybeAgentObjetive = option{TargetAgent}
                set MaybePrimary_TargetPosition = option{NewTargetCharacter.GetTransform().Translation}
                #Secondary target
                Secondary_TargetPosition := MobBehaviourHelper.FindRandomNavigationPoint(NPCCharacter.GetTransform().Translation, GroundMob_Max_Movement,NPCSpawnPoint.Translation)
                set MaybeSecondary_TargetPosition = option{Secondary_TargetPosition}
            else:
                #Creation two posible positions to move
                Primary_TargetPosition := MobBehaviourHelper.FindRandomNavigationPoint(NPCCharacter.GetTransform().Translation, GroundMob_Max_Movement,NPCSpawnPoint.Translation)
                set MaybePrimary_TargetPosition = option{Primary_TargetPosition}
                Secondary_TargetPosition := MobBehaviourHelper.FindRandomNavigationPoint(NPCCharacter.GetTransform().Translation, GroundMob_Max_Movement,NPCSpawnPoint.Translation)
                set MaybeSecondary_TargetPosition = option{Secondary_TargetPosition}


        return (MaybeAgentObjetive,MaybePrimary_TargetPosition,MaybeSecondary_TargetPosition,IsClippingDetected, Mob_NavigationErrorWait,ReachRadius)
    ##############
    #DUPLICATED MAKE ACTION
    MakeAction(NPCTransform : transform, TargetTransform : transform)<suspends> : tuple(logic,logic,?creative_prop,float,float,?animation_sequence) =
        var IsSpecialActionExecuted : logic = false
        var MustAttack : logic = false
        var PreActionDelay : float = 1.0
        var PostActionDelay : float = 1.0
        var MaybeAreaCreative_Prop : ?creative_prop = false
        var MaybeAnimation : ?animation_sequence = false
            if(THIS_MOBTYPE=MINIBOSS_MOB, GeneralMobsManager := MobBehaviourHelper.MaybeGeneralMobsManager?): 
                NumDungeon:=GeneralMobsManager.FindNearDungeon(NPCTransform.Translation)
                GeneralMobsManager.CheckBossMusic(NumDungeon)
                if(NumSpecialActions<MiniBoss_MaxSpecialActions, MobBehaviourHelper.TotalTime>LastSpecialAction+MiniBoss_WaitBetweenSpecialActions):
                    #Checking num of mobs spawned in this dungeon
                    var Num_Child_Spawned : int = 0
                    var Num_NoChild_Spawned : int = 0
                    var AllNoChildMobsSpawned : []mob = array{}
                    var AllChildMobsSpawned : []mob  = array{}
                    var IsSpawnMinionsAllowed : logic = false
                    for(Mob : GeneralMobsManager.AllMobsManager, Mob.MobType<>MINIBOSS_MOB,not Mob.DeathRunning?, Mob.IsBeingUsed?, Mob.NumDungeon=NumDungeon):
                        if(Mob.IsBossChild?):
                            set Num_Child_Spawned = Num_Child_Spawned + 1
                            set AllChildMobsSpawned = AllChildMobsSpawned + array{Mob}
                        else:
                            set Num_NoChild_Spawned = Num_NoChild_Spawned + 1
                            set AllNoChildMobsSpawned = AllNoChildMobsSpawned + array{Mob}
                    #Checking number of mobs in this dungeon
                    #Print("NumMobsSpawned: {Num_Child_Spawned}")
                    #Print("Num No Child: {Num_NoChild_Spawned}")
                    if(Num_NoChild_Spawned>10):
                        for(Mob : AllNoChildMobsSpawned):
                            #Print("Destroying  no child minions there are too many")
                            Mob.DealDamage(999999999999999999999.9)
                        set IsSpawnMinionsAllowed = true
                    else if(Num_NoChild_Spawned > 4, Num_Child_Spawned > 4):
                        for(Mob : AllNoChildMobsSpawned):
                            #Print("Destroying  no child minions because there is not space")
                            Mob.DealDamage(999999999999999999999.9)
                        set IsSpawnMinionsAllowed = true
                    else if(Num_Child_Spawned > 9):
                        for(Mob : AllNoChildMobsSpawned):
                            #Print("Destroying  no child minions for space")
                            Mob.DealDamage(999999999999999999999.9)
                        set IsSpawnMinionsAllowed = true

                    if(IsSpawnMinionsAllowed?):
                        MobBehaviourHelper.PlaySound(NPCTransform,SOUND_BOSS_LAUGH,1.0)
                        set IsSpecialActionExecuted = true
                        #Print("Special action, spawn minions")
                        spawn:
                            MobBehaviourHelper.SpawnMiniBossMinions(NPCTransform,NPCSpawnPoint.Translation)
                        Sleep(MiniBoss_MinionsSpawnToWait_Duration)
                else:
                    if(300.0>DistanceXY(TargetTransform.Translation,NPCTransform.Translation)):
                        if(0=GetRandomInt(0,1)):
                            set PreActionDelay=MiniBossAttack_PreDelay
                            set PostActionDelay=MiniBossAttack_PostDelay
                            set MustAttack = true
                            set MaybeAnimation = option{MiniBossMobAttack}
                        else:
                            MobBehaviourHelper.CreateDamageZone(NPCTransform)
                    else:
                        MobBehaviourHelper.CreateDamageZone(NPCTransform)
            else if(THIS_MOBTYPE=DRUID_MOB):
                if(MobBehaviourHelper.TotalTime>LastSpecialAction+Druid_WaitBetweenSpecialActions):
                    set IsSpecialActionExecuted = true
                    #Print("Special action, spawn healing all mob allies in the area")
                    HealingZoneTransform:=NPCTransform
                    spawn:
                        MobBehaviourHelper.PlaySound(NPCTransform,SOUND_LOAD_ZONE,DamageZonePreDamage_Delay)
                    set MaybeAreaCreative_Prop = MobBehaviourHelper.SpawnMobHealZone(NPCTransform)
                    Sleep(HealZonePreDamage_Delay)
                    spawn:
                        MobBehaviourHelper.PlaySound(NPCTransform,SOUND_HEAL_ZONE,-1.0)
                    if(GeneralMobsManager := MobBehaviourHelper.MaybeGeneralMobsManager?):
                        for(Mob : GeneralMobsManager.AllMobsManager, not Mob.DeathRunning?, Mob.IsBeingUsed?):
                            if(OtherNPCCharacter := Mob.MaybeNPCCharacter?, Transform := OtherNPCCharacter.GetTransform()):
                                if(MobHealZoneRadius>DistanceXY(OtherNPCCharacter.GetTransform().Translation,HealingZoneTransform.Translation)):
                                    OtherNPCCharacter.Heal(OtherNPCCharacter.GetMaxHealth()*0.33)
                                    #Print("Character on zone detected, healing him")
                        Sleep(HealZonePostDamage_Delay)
                else:
                    if(DruidMobAttack_Range>DistanceXY(TargetTransform.Translation,NPCTransform.Translation)):
                        set PreActionDelay=DruidMobAttack_PreDelay
                        set PostActionDelay=DruidMobAttack_PostDelay
                        set MustAttack = true
                        set MaybeAnimation = option{DruidMobAttack}
            else if(THIS_MOBTYPE=GROUND_MOB):
                if(GroundMobAttack_Range>DistanceXY(TargetTransform.Translation,NPCTransform.Translation)):
                    set PreActionDelay=ShortAttack_PreDelay
                    set PostActionDelay=ShortAttack_PostDelay
                    set MustAttack = true
                    set MaybeAnimation = option{GroundMobAttack}

            else if(THIS_MOBTYPE=PLANT_MOB):
                if(PlantMobAttack_Range>DistanceXY(TargetTransform.Translation,NPCTransform.Translation)):
                    set PreActionDelay=LongAttack_PreDelay
                    set PostActionDelay=LongAttack_PostDelay
                    set MustAttack = true
                    set MaybeAnimation = option{PlantMobAttack}
        else:
            #There is no player on area...waiting
            Sleep(PlantMob_MoveToWait_Duration)
            Sleep(0.1)
        return (IsSpecialActionExecuted,MustAttack,MaybeAreaCreative_Prop,PreActionDelay,PostActionDelay,MaybeAnimation)

    ###########
#DUPLICATED CODE#>

    FindTag() : void =
        ##Print("Tring to find objects")
        if(ADevice:=MyVariable[GetSession()]):
            AllCreativeObjects:=ADevice.FindCreativeObjectsWithTag(helper{})
            ##Print("Creative objects detected")
            for(CreativeObject:AllCreativeObjects):
                ##Print("Single Creative object")
                if(Helper := mob_behavior_helper[CreativeObject]):
                    ##Print("Tag object MaybeMobBehaviourHelper found")
                    set MaybeMobBehaviourHelper = option{Helper} 
                    set LastSpecialAction = Helper.TotalTime-99
            #else:
            #    PrintFatalError("Tag object MaybeMobBehaviourHelper not found")

    SetDamageToPlayer() : void =
        #Now, set the damage
        if(Helper := MaybeMobBehaviourHelper?, GeneralMobsManager:=Helper.MaybeGeneralMobsManager?):
            var IsModFound : logic = false
            for(Mob : GeneralMobsManager.AllMobsManager, ThisNPC := MaybeCharacter?, OtherNPCCharacter := Mob.MaybeNPCCharacter?, 1.0>Distance(OtherNPCCharacter.GetTransform().Translation, ThisNPC.GetTransform().Translation)):
                Log("Mob found!",LOG_LEVEL_DEBUG7)
                DamageHitTuple := GeneralMobsManager.CalculateMobDamage(THIS_MOBTYPE,Mob.HBAreaLV,Mob.NumDungeon)
                set Main_DamageToPlayer = DamageHitTuple(0)
                set Secondary_DamageToPlayer = DamageHitTuple(1)
                set IsModFound = true
            if(not IsModFound?):
                PrintFatalError("Mob NOT found!")
            else:
                Log("Mob found!, Type: {THIS_MOBTYPE}, MainDamage: {GetOptimizedMagnitudeLetter(Main_DamageToPlayer)}, SecondaryDamage: {GetOptimizedMagnitudeLetter(Secondary_DamageToPlayer)}",LOG_LEVEL_DEBUG7)
            #else:
            #    PrintFatalError("Tag object MaybeMobBehaviourHelper not found")

    SpawnMiniBossInvulnerability(Duration : float)<suspends> : void =
        if(MobBehaviourHelper := MaybeMobBehaviourHelper?,MiniBossNPC:=MaybeCharacter?,MiniBossNPC.IsActive[]):
            Transform:=MiniBossNPC.GetTransform()
            ShieldSpawnTransform:=transform{Translation:=Transform.Translation+vector3{Z:=-82.0},Rotation:=Transform.Rotation,Scale:=Transform.Scale}
            TupleSpawn:=SpawnProp(MobBehaviourHelper.BP_MiniBossShieldZone,ShieldSpawnTransform)
            MaybeCreativeProp:=TupleSpawn(0)
            if(ValidShieldProp := MaybeCreativeProp?):
                spawn:
                    MobBehaviourHelper.DisposeEffect(ValidShieldProp,10.0)
                MiniBossNPC.SetVulnerability(false)
                var Time : float = 0.0
                TickTime := 0.2
                loop:
                    NewShieldSpawnTransform:=transform{Translation:=MiniBossNPC.GetTransform().Translation+vector3{Z:=-82.0},Rotation:=MiniBossNPC.GetTransform().Rotation,Scale:=Transform.Scale}
                    if(MiniBossNPC.IsActive[],MiniBossNPC.GetHealth()>0.0, ValidShieldProp.IsValid[], ValidShieldProp.TeleportTo[NewShieldSpawnTransform]):
                    else:
                        break
                    Sleep(TickTime)
                    set Time = Time + TickTime
                    if(Time>Duration):
                        break
                if(ValidShieldProp.IsValid[]):
                    ValidShieldProp.Dispose()
                MiniBossNPC.SetVulnerability(true)

    # This function runs when the NPC is spawned in the world and ready to follow a behavior.
    OnBegin<override>()<suspends>:void = 
        
        <#
        AllObjects:=FindCreativeObjectsWithTag(helper{})
        for(Object : AllObjects):
            if(Helper := mob_behavior_helper[Object]):
                set MobBehaviourHelper = Helper
                set LastSpecialAction = MobBehaviourHelper.TotalTime-99
            else:
                PrintFatalError("FATAL ERROR on miniboss_mob_behaviour, CreativeObjectWithTag ``helper`` not found")#>

        FindTag()

        if:
            # Get the Agent (the NPC).
            NPCAgent := GetAgent[]

            # Gets the Fortnite Character interface, which gets you access to its gameplay data 
            # including its AI module for navigation and focus.
            NPCCharacter := NPCAgent.GetFortCharacter[] 

            # Get the Navigatable Interface, this allows you to tell it to move.
            NPCNavigatable := NPCCharacter.GetNavigatable[]

            # Get the Focus Interface, this allows you to tell it to look at something or somewhere.
            NPCFocus := NPCCharacter.GetFocusInterface[]

            NPCAnim := NPCCharacter.GetPlayAnimationController[]

            if(MobBehaviourHelper := MaybeMobBehaviourHelper?, THIS_MOBTYPE=MINIBOSS_MOB, set LastSpecialAction = MobBehaviourHelper.TotalTime):

        then:
            set MaybeCharacter = option{NPCCharacter}
            Sleep(0.2)
            SetDamageToPlayer()
            # TODO: Now that you have the Fortnite Character and the AI interfaces, replace this with your code 
            # for your desired movement, look at, and other behavior allowed by the AI module.
            if(ShowAIDebug?, LOG_LEVEL=LOG_LEVEL_NPCDEBUG):
                Print(ground_mob_behaviour_message_module.OnBeginMessage(NPCAgent), ?Duration := AIDebugDrawTime)

            # Save the position of your character which we'll use to generate more points to move to from.
            set NPCSpawnPoint = NPCCharacter.GetTransform()

            loop:
                #Move MOB
                var IsNavigationError : logic = false
                if(MobBehaviourHelper := MaybeMobBehaviourHelper?, NPCCharacter.GetHealth()>0.0, NPCCharacter.IsActive[]):
                    TupleMovementDecision:=MobBehaviourHelper.MakeMovementDecision(NPCAgent,NPCCharacter,THIS_MOBTYPE,NPCSpawnPoint)
                    MaybeAgentObjetive := TupleMovementDecision(0)
                    MaybePrimary_TargetPosition := TupleMovementDecision(1)
                    MaybeSecondary_TargetPosition := TupleMovementDecision(2)
                    IsClippingDetected := TupleMovementDecision(3)
                    Mob_NavigationErrorWait := TupleMovementDecision(4)
                    ReachRadius := TupleMovementDecision(5)
                    
                    #Making movement
                    if(NPCCharacter.IsActive[], TargetAgent:=MaybeAgentObjetive?):
                        ##Print("Changing focus")
                        spawn{NPCFocus.MaintainFocus(TargetAgent)}
                    if:
                        Primary_TargetPosition := MaybePrimary_TargetPosition?
                        Secondary_TargetPosition := MaybeSecondary_TargetPosition?
                    then:
                        Primary_NPCNavigationTarget := MakeNavigationTarget(vector3{X:=Primary_TargetPosition.X,Y:=Primary_TargetPosition.Y,Z:=Primary_TargetPosition.Z})
                        NavResultGoTo1:=NPCNavigatable.NavigateTo(Primary_NPCNavigationTarget, ?MovementType := movement_types.Walking, ?ReachRadius := ReachRadius)
                        if(IsClippingDetected?):
                            if(THIS_MOBTYPE=MINIBOSS_MOB):
                                MobBehaviourHelper.PlayDirtEffect(NPCCharacter.GetTransform())
                                if(NPCCharacter.TeleportTo[NPCSpawnPoint.Translation, NPCSpawnPoint.Rotation]):
                                    MobBehaviourHelper.PlayDirtEffect(NPCSpawnPoint)
                                    Log("Navigation error2",LOG_LEVEL_NPCDEBUG)
                                else:
                                    set IsNavigationError = true
                                    Sleep(Mob_NavigationErrorWait)
                        else:
                            if(NavResultGoTo1 <> navigation_result.Reached):
                                #try to navigate to point 2
                                Secondary_NPCNavigationTarget := MakeNavigationTarget(vector3{X:=Secondary_TargetPosition.X,Y:=Secondary_TargetPosition.Y,Z:=Secondary_TargetPosition.Z})
                                NavResultGoTo2:=NPCNavigatable.NavigateTo(Primary_NPCNavigationTarget, ?MovementType := movement_types.Walking, ?ReachRadius := ReachRadius) 
                                if(NavResultGoTo2 <> navigation_result.Reached):
                                    Log("Navigation error1",LOG_LEVEL_NPCDEBUG)
                                    #If there is again an error, teleport to spawn point
                                    MobBehaviourHelper.PlayDirtEffect(NPCCharacter.GetTransform())
                                    if(NPCCharacter.TeleportTo[NPCSpawnPoint.Translation, NPCSpawnPoint.Rotation]):
                                        Log("Navigation error2",LOG_LEVEL_NPCDEBUG)
                                        MobBehaviourHelper.PlayDirtEffect(NPCSpawnPoint)
                                    else:
                                        #Can not teleport? make an error
                                        set IsNavigationError = true
                                        Sleep(Mob_NavigationErrorWait)
                    if(IsNavigationError?):
                        set NumNavigationErrors = NumNavigationErrors + 1
                        if(NumNavigationErrors>10):
                            PrintFatalError("Too many errors, destroying")
                            NPCCharacter.Damage(99999999999.9)
                #Make MOB action
                if(MobBehaviourHelper := MaybeMobBehaviourHelper?, NPCCharacter.IsActive[], NPCCharacter.GetHealth()>0.0):
                    if(NewTargetCharacter:=MobBehaviourHelper.FindNearestTargetFromPos[NPCCharacter,MiniBoss_CheckArena]):
                        TupleActionResult:=MobBehaviourHelper.MakeActionV2(NPCCharacter.GetTransform(),NewTargetCharacter,THIS_MOBTYPE,NPCSpawnPoint,LastSpecialAction,NumSpecialActions,Secondary_DamageToPlayer)
                        IsSpecialActionExecuted:=TupleActionResult(0)
                        MustAttack:=TupleActionResult(1)
                        PreAttackDelay:=TupleActionResult(3)
                        PostAttackDelay:=TupleActionResult(4)
                        MaybeAnimation:=TupleActionResult(5)
                        IsZoneAttack:=TupleActionResult(6)
                        if(MustAttack?, NPCAnimation := MaybeAnimation?):
                            NPCAnim.Play(NPCAnimation,?BlendInTime:=0.25,?BlendOutTime:=0.25)
                            MobBehaviourHelper.HandlerPlayBite(THIS_MOBTYPE,NPCCharacter.GetTransform(),PreAttackDelay)
                            Sleep(PreAttackDelay)
                            if(NPCCharacter.GetHealth()>0.0):
                                MobBehaviourHelper.MakeSpecialRay_DamageForward(NPCCharacter.GetTransform(),NPCCharacter.GetViewLocation(),NPCCharacter.GetViewRotation(),THIS_MOBTYPE,Main_DamageToPlayer)
                            Sleep(PostAttackDelay)
                        if(IsZoneAttack?):
                            Sleep(PreAttackDelay)
                            Sleep(PostAttackDelay)
                        if(IsSpecialActionExecuted?, NPCAnimation := MaybeAnimation?):
                            NPCAnim.Play(NPCAnimation,?BlendInTime:=0.25,?BlendOutTime:=0.25)
                            Sleep(PreAttackDelay)
                            SpawnMiniBossInvulnerability(7.0)
                            Sleep(PostAttackDelay)
                            set NumSpecialActions = NumSpecialActions+1
                            set LastSpecialAction = MobBehaviourHelper.TotalTime
                    else:
                        Sleep(0.1)
                else:
                    Sleep(0.1)

                <#
                #Make MOB action
                if(NPCCharacter.IsActive[], NPCCharacter.GetHealth()>0.0):
                    if(NewTargetCharacter:=MobBehaviourHelper.FindNearestTargetFromPos[NPCCharacter,MiniBoss_CheckArena]):
                        TupleActionResult:=MobBehaviourHelper.MakeAction(NPCCharacter.GetTransform(),NewTargetCharacter,THIS_MOBTYPE,NPCSpawnPoint,LastSpecialAction,NumSpecialActions)
                        IsSpecialActionExecuted:=TupleActionResult(0)
                        MustAttack:=TupleActionResult(1)
                        PreAttackDelay:=TupleActionResult(3)
                        PostAttackDelay:=TupleActionResult(4)
                        MaybeAnimation:=TupleActionResult(5)
                        if(MustAttack?, NPCAnimation := MaybeAnimation?):
                            NPCAnim.Play(NPCAnimation,?BlendInTime:=0.25,?BlendOutTime:=0.25)
                            spawn:
                                MobBehaviourHelper.PlayBite(THIS_MOBTYPE,NPCCharacter.GetTransform(),PreAttackDelay)
                            Sleep(PreAttackDelay)
                            if(NPCCharacter.GetHealth()>0.0):
                                MobBehaviourHelper.MakeSpecialRay_DamageForward(NPCCharacter.GetTransform(),NPCCharacter.GetViewLocation(),NPCCharacter.GetViewRotation(),THIS_MOBTYPE)
                            Sleep(PostAttackDelay)
                        if(IsSpecialActionExecuted?):
                            SpawnMiniBossInvulnerability(7.0)
                            set NumSpecialActions = NumSpecialActions+1
                            set LastSpecialAction = MobBehaviourHelper.TotalTime
                    else:
                        Sleep(0.1)
                else:
                    Sleep(0.1)#>

                
        else:
            # If code falls here something failed when gathering the agent or its interfaces
            if(ShowAIDebug?):
                PrintNPCB("Error in NPC Behavior Script on NPC Setup",
                        ?Duration := AIDebugDrawTime, 
                        ?TextColor := NamedColors.Red )

    # This function draws a box around a specified position for a finite amount of time.
    # NOTE: To see this in game, Verse Debug Draw must be enabled in Island Settings.
    DrawDebugLocation(Location:vector3):void = 
        DebugDrawNPC.DrawPoint( Location, 
                                ?Color := NamedColors.SteelBlue, 
                                ?Thickness := 100.0, 
                                ?DrawDurationPolicy := debug_draw_duration_policy.FiniteDuration, 
                                ?Duration := AIDebugDrawTime )

    # This function draws an arrow from the Agent's head to its look at point every half a second. 
    # NOTE: To see this in game, Verse Debug Draw must be enabled in Island Settings.
    DrawDebugLookAt(Character:fort_character, LookAtPoint:vector3)<suspends>:void=
        loop:
            DebugDrawNPC.DrawArrow( Character.GetTransform().Translation + vector3{ Z := VerticalOffsetToNPCHead}, 
                                    LookAtPoint, 
                                    ?ArrowSize := 50.0, 
                                    ?Color := NamedColors.Yellow, 
                                    ?Thickness := 5.0, 
                                    ?DrawDurationPolicy := debug_draw_duration_policy.FiniteDuration, 
                                    ?Duration := LookAtDebugDrawDuration )
                                    
            #This sleep matches the length of time the arrow is drawn to avoid duplicate draws.
            Sleep(LookAtDebugDrawDuration)

    # Custom wrapper that provides a default duration and color.
    PrintNPCB(Msg:string,?Duration:float = AIDebugDrawTime, ?TextColor:color = NamedColors.Green):void =
        Print("[ground_mob_behaviour] {Msg}", ?Color := TextColor, ?Duration := Duration)

    # This function runs when the NPC is despawned or eliminated from the world.
    OnEnd<override>():void =
        # TODO: Replace this with your code for any cleanup that you want to happen.
        if(Agent := GetAgent[], LOG_LEVEL=LOG_LEVEL_NPCDEBUG):
            Print(ground_mob_behaviour_message_module.OnEndMessage(Agent), ?Duration := AIDebugDrawTime)
        else:
            #PrintNPCB("OnEnd")


            
