using { /Fortnite.com/Devices }
using { /Verse.org/Simulation }
using { /UnrealEngine.com/Temporary/Diagnostics }
using { /UnrealEngine.com/Temporary/SpatialMath }
using { /Verse.org/Random }
using { /Fortnite.com/Characters }

# See https://dev.epicgames.com/documentation/en-us/uefn/create-your-own-device-in-verse for how to create a verse device.

# A Verse-authored creative device that can be placed in a level
minigames_manager := class(creative_device):

    ArrowCategory<public><localizes>:message := "Arrow MiniGame"
    SkilledCategory<public><localizes>:message := "Skilled Interaction MiniGame"
    SimonCategory<public><localizes>:message := "Simon"
    InestabilityCategory<public><localizes>:message := "Inestable HB"
    SteveGamesCategory<public><localizes>:message := "Steve"

    @editable
    MiniGameCoolDown : int = 600

#MINIGAMES INTS
    SKILLEDPREASSURE_MINIGAME : int = 0
    SKILLEDELECTRICAL_MINIGAME : int = 1
    SKILLEDTRUCK_MINIGAME : int = 2
    ARCHERY0_MINIGAME : int = 3
    ARCHERY1_MINIGAME : int = 4
    ARCHERY2_MINIGAME : int = 5
    ARCHERY3_MINIGAME : int = 6
    ARCHERY4_MINIGAME : int = 7
    ARCHERY5_MINIGAME : int = 8
    ARCHERY6_MINIGAME : int = 9
    ARCHERY7_MINIGAME : int = 10
    ARCHERY8_MINIGAME : int = 11
    SIMONCOLOR0_MINIGAME : int = 12
    SIMONCOLOR1_MINIGAME : int = 13
    SIMONCOLOR2_MINIGAME : int = 14
    SIMONCOLOR3_MINIGAME : int = 15
    SIMONCOLOR4_MINIGAME : int = 16
    SIMONCOLOR5_MINIGAME : int = 17
    SIMONCOLOR6_MINIGAME : int = 18
    SIMONCOLOR7_MINIGAME : int = 19
    SIMONCOLOR8_MINIGAME : int = 20
    SIMONCOLOR9_MINIGAME : int = 21
    SIMONSOUND0_MINIGAME : int = 22
    SIMONSOUND1_MINIGAME : int = 23
    SIMONSOUND2_MINIGAME : int = 24
    SIMONSOUND3_MINIGAME : int = 25
    SIMONSOUND4_MINIGAME : int = 26
    SIMONSOUND5_MINIGAME : int = 27
    SIMONSOUND6_MINIGAME : int = 28
    SIMONSOUND7_MINIGAME : int = 29
    SIMONSOUND8_MINIGAME : int = 30
    SIMONSOUND9_MINIGAME : int = 31
    SIMONSOUND10_MINIGAME : int = 32
    SIMONSOUND11_MINIGAME : int = 33
    STEVE_SKILLED0_MINIGAME : int = 34
    STEVE_SKILLED1_MINIGAME : int = 35
    STEVE_SKILLED2_MINIGAME : int = 36

#SCROLL FIND
    @editable
    AllScrolls : []collectible_object_device = array{}
    @editable
    AllScrollPositions_Area0 : []creative_prop = array{}
    @editable
    AllScrollPositions_Area2 : []creative_prop = array{}
    @editable
    AllScrollPositions_Area8 : []creative_prop = array{}
    var AllScrollsOriginalPositions : []transform = array{}

#SIMON
    #SIMON MODES
    var AllSimonPlayerLevel : []int = array{0,0,0,0}
    var AllSimonPlayerReplies : [][]int = array{array{},array{},array{},array{}}
    var AllSimonSays : [][]int = array{}
    var AllSimonSoundAssignation:[]audio_player_device = array{}
    SIMON_COLOR : int = 0
    SIMON_SOUND : int = 1

    SIMON_RED : int = 0
    SIMON_GREEN : int = 1
    SIMON_BLUE : int = 2
    SIMON_WHITE : int  = 3
    SIMON_YELLOW : int = 4

    var SimonCoolDown : int = 0
    SimonMaxCoolDown : int = 600
    SimonMaxTime : int = 240
    var SimonTotalTime : int = 0
    var SimonGameActive : int = -1
    var SimonTeleporterOrigin : transform = transform{}
    var AllSimonIsTalking : []logic = array{false,false,false,false}
    @editable:
        Categories := array{SimonCategory} 
    AllSimon_Positions : []creative_prop = array{}
    @editable:
        Categories := array{SimonCategory} 
    AllPosibleSimonTeleporter_Positions : []creative_prop = array{}
    @editable:
        Categories := array{SimonCategory} 
    AllSimonButtons : []button_device = array{}
    @editable:
        Categories := array{SimonCategory} 
    TeleporterSimonProp : creative_prop = creative_prop{}
    @editable:
        Categories := array{SimonCategory} 
    TeleporterSimonArea : mutator_zone_device = mutator_zone_device{}
    @editable:
        Categories := array{SimonCategory} 
    AllSimonTeleporters : []mutator_zone_device = array{}
    @editable:
        Categories := array{SimonCategory} 
    AllSimonRed : []creative_prop = array{}
    @editable:
        Categories := array{SimonCategory} 
    AllSimonRedMutator : []prop_manipulator_device = array{}
    @editable:
        Categories := array{SimonCategory} 
    AllSimonWhite: []creative_prop = array{}
    @editable:
        Categories := array{SimonCategory} 
    AllSimonWhiteMutator : []prop_manipulator_device = array{}
    @editable:
        Categories := array{SimonCategory} 
    AllSimonGreen : []creative_prop = array{}
    @editable:
        Categories := array{SimonCategory} 
    AllSimonGreenMutator : []prop_manipulator_device = array{}
    @editable:
        Categories := array{SimonCategory} 
    AllSimonYellow : []creative_prop = array{}
    @editable:
        Categories := array{SimonCategory} 
    AllSimonYellowMutator : []prop_manipulator_device = array{}
    @editable:
        Categories := array{SimonCategory} 
    AllSimonBlue : []creative_prop = array{}
    @editable:
        Categories := array{SimonCategory} 
    AllSimonBlueMutator : []prop_manipulator_device = array{}
    @editable:
        Categories := array{SimonCategory} 
    Do_AudioPlayer : audio_player_device = audio_player_device{}
    @editable:
        Categories := array{SimonCategory} 
    Re_AudioPlayer : audio_player_device = audio_player_device{}
    @editable:
        Categories := array{SimonCategory} 
    Mi_AudioPlayer : audio_player_device = audio_player_device{}
    @editable:
        Categories := array{SimonCategory} 
    Fa_AudioPlayer : audio_player_device = audio_player_device{}
    @editable:
        Categories := array{SimonCategory} 
    Sol_AudioPlayer : audio_player_device = audio_player_device{}


#SKILLED INTERACTION DEVICE
    @editable:
        Categories := array{SkilledCategory}
    Skilled_CoalReward : float = 5.0
    @editable:
        Categories := array{SkilledCategory} 
    SkilledInteraction_Truck: skilled_interaction_device = skilled_interaction_device{}
    @editable:
        Categories := array{SkilledCategory}
    Button_SkilledInteraction_Truck : button_device = button_device{}
    @editable:
        Categories := array{SkilledCategory} 
    SkilledInteraction_Electrical : skilled_interaction_device = skilled_interaction_device{}
    @editable:
        Categories := array{SkilledCategory}
    Button_SkilledInteraction_Electrical : button_device = button_device{}
    @editable:
        Categories := array{SkilledCategory} 
    SkilledInteraction_Preassure : skilled_interaction_device = skilled_interaction_device{}
    @editable:
        Categories := array{SkilledCategory}
    Button_SkilledInteraction_Preassure : button_device = button_device{}
    

#ARROWS MINIGAME
    ArcheryPillarsTranslation : float = 1524.0
    var ArcheryMiniGame_Regeneration: float = 0.05
    var MaybeGameManager : ?game_manager = false
    var AllPillars : [][]creative_prop = array{}
    var AllPillarsOriginalPositions : [][]transform = array{}
    var AllArcheryState : []int = array{}
    var AllArcheryMiniGame_MaxHealth : []float =  array{}
    var AllArcheryMiniGame_CurrentHealth : []float =  array{}
    var AllArcheryMiniGame_RegenerationActive : []logic = array{}
    var AllArcheryDamagingAgents : [][agent]int = array{}

    @editable:
        Categories := array{ArrowCategory}
    Archery_CoalReward_PerLevel : float = 10.0
    @editable:
        Categories := array{ArrowCategory}
    AllArcheryManipulatorDevices : []prop_manipulator_device = array{}
    @editable:
        Categories := array{ArrowCategory} 
    AllArcheryPillarsMaxTime : []int = array{10,20,20,20,20,20,20,20,20,20,20}
    #Arrows difficulty
    @editable:
        Categories := array{ArrowCategory} 
    AllArcheryButtons : []button_device = array{}
        #Arrow difficulty 0
    @editable:
        Categories := array{ArrowCategory} 
    AllPillars0: []creative_prop = array{}
    @editable:
        Categories := array{ArrowCategory}
    Difficulty0 : int = 1
    var AllArcheryDamagingAgents0 : [agent]int = map{}
        #Arrow difficulty 1
    @editable:
        Categories := array{ArrowCategory} 
    AllPillars1: []creative_prop = array{}
    @editable:
        Categories := array{ArrowCategory}
    Difficulty1 : int = 2
    var AllArcheryDamagingAgents1 : [agent]int = map{}
    #Arrow difficulty 2
    @editable:
        Categories := array{ArrowCategory} 
    AllPillars2: []creative_prop = array{}
    @editable:
        Categories := array{ArrowCategory}
    Difficulty2 : int = 3
    var AllArcheryDamagingAgents2 : [agent]int = map{}
    #Arrow difficulty 3
    @editable:
        Categories := array{ArrowCategory} 
    AllPillars3: []creative_prop = array{}
    @editable:
        Categories := array{ArrowCategory}
    Difficulty3 : int = 4
    var AllArcheryDamagingAgents3 : [agent]int = map{}
    #Arrow difficulty 4
    @editable:
        Categories := array{ArrowCategory} 
    AllPillars4: []creative_prop = array{}
    @editable:
        Categories := array{ArrowCategory}
    Difficulty4 : int = 4
    var AllArcheryDamagingAgents4 : [agent]int = map{}
    #Arrow difficulty 5
    @editable:
        Categories := array{ArrowCategory} 
    AllPillars5: []creative_prop = array{}
    @editable:
        Categories := array{ArrowCategory}
    Difficulty5 : int = 5
    var AllArcheryDamagingAgents5 : [agent]int = map{}
    #Arrow difficulty 6
    @editable:
        Categories := array{ArrowCategory} 
    AllPillars6: []creative_prop = array{}
    @editable:
        Categories := array{ArrowCategory}
    Difficulty6 : int = 6
    var AllArcheryDamagingAgents6 : [agent]int = map{}
    #Arrow difficulty 7
    @editable:
        Categories := array{ArrowCategory} 
    AllPillars7: []creative_prop = array{}
    @editable:
        Categories := array{ArrowCategory}
    Difficulty7: int = 7
    var AllArcheryDamagingAgents7 : [agent]int = map{}

#INESTABILITY HARD BREADS
    var AllInestableHB : []hardbread_managerv2 = array{}
    @editable:
        Categories := array{InestabilityCategory}
    AllInestablePosiblePositions : []creative_prop = array{}
    @editable:
        Categories := array{InestabilityCategory}
    MaxHBInestableTime : int = 120

#SCROLL FIND
    RemoveScroll(AgentIndex : int) : void =
        if(Scroll := AllScrolls[AgentIndex]):
            if(OriginalTransform:=AllScrollsOriginalPositions[AgentIndex]):
                if(Distance(OriginalTransform.Translation,Scroll.GetTransform().Translation)>50.0):
                    Log("Teleporting back the Scroll",LOG_LEVEL_DEBUG6)
                    if(Scroll.TeleportTo[OriginalTransform]):
                        Scroll.Hide()
                    else:
                        spawn:
                            Scroll.MoveTo(OriginalTransform.Translation,OriginalTransform.Rotation,1.0)

    CheckScrollSpawn() : void =
        if(GameManager:=MaybeGameManager?):
            for(Player : GetPlayspace().GetPlayers(), Player.IsActive[], Agent := agent[Player], AgentBakery := GameManager.GetAgentBakery_Transacted(Agent)?):
                AgentIndex := GameManager.GetAgentIndex(Agent)
                if(Scroll := AllScrolls[AgentIndex]):
                    var AllScrollPositions : []creative_prop = array{}

                    if(AgentBakery.PlayerDataCache.QuestAssigned_ObjetiveType=FINDSCROLL_AREA0_QUEST_TYPE):
                        set AllScrollPositions = AllScrollPositions_Area0
                    else if(AgentBakery.PlayerDataCache.QuestAssigned_ObjetiveType=FINDSCROLL_AREA2_QUEST_TYPE):
                        set AllScrollPositions = AllScrollPositions_Area2
                    else if(AgentBakery.PlayerDataCache.QuestAssigned_ObjetiveType=FINDSCROLL_AREA8_QUEST_TYPE):
                        set AllScrollPositions = AllScrollPositions_Area8

                    if(AllScrollPositions.Length>0):
                        Scroll.RespawnForAll()
                        var IsAlreadyOnPosition : logic = false
                        for(Position : AllScrollPositions):
                            Translation := AdaptTranslationFromVisualEmpty(Position.GetTransform().Translation,VisualEmptyHideZ)
                            if(Distance(Translation,Scroll.GetTransform().Translation)<256.0):
                                set IsAlreadyOnPosition = true
                        if(not IsAlreadyOnPosition?):
                            RandomPositionIndex := GetRandomInt(0,AllScrollPositions.Length-1)
                            if(PositionAssigned := AllScrollPositions[RandomPositionIndex]):
                                TransformAssigned := AdaptTransformFromVisualEmpty(PositionAssigned.GetTransform(),VisualEmptyHideZ+64.0)
                                if(Scroll.TeleportTo[TransformAssigned]):
                                    Scroll.RespawnForAll()
                                else:
                                    TransformAssignedV2 := AdaptTransformFromVisualEmpty(PositionAssigned.GetTransform(),VisualEmptyHideZ+84.0)
                                    if(Scroll.TeleportTo[TransformAssignedV2]):
                                        Scroll.RespawnForAll()
                                    else:
                                        PrintWarning("Failed teleporting the scroll. Completing the player Quest")
                                        var UpdatedDataTable : player_data_table = AgentBakery.GameManager.QuestsAndOrdersDatabase.AddAgentQuest_ProgressV2(Player,AgentBakery.PlayerDataCache,1.0,FINDSCROLL_AREA0_QUEST_TYPE)(4)
                                        set UpdatedDataTable = AgentBakery.GameManager.QuestsAndOrdersDatabase.AddAgentQuest_ProgressV2(Player,UpdatedDataTable,1.0,FINDSCROLL_AREA2_QUEST_TYPE)(4)
                                        set UpdatedDataTable = AgentBakery.GameManager.QuestsAndOrdersDatabase.AddAgentQuest_ProgressV2(Player,UpdatedDataTable,1.0,FINDSCROLL_AREA8_QUEST_TYPE)(4)
                                        AgentBakery.SetNewPlayerDataCache(UpdatedDataTable)
                                        GetOptimizedMagnitudeLetter(X:=TransformAssigned.Translation.X)
                                
                                Log("Spawning Scroll on position {RandomPositionIndex}, Position: X:{GetOptimizedMagnitudeLetter(TransformAssigned.Translation.X)}, Y:{GetOptimizedMagnitudeLetter(TransformAssigned.Translation.Y)}, Z:{GetOptimizedMagnitudeLetter(TransformAssigned.Translation.Z)}",LOG_LEVEL_DEBUG6)
                    else:
                        RemoveScroll(AgentIndex)





    ScrollCollectedEvent(Agent : agent) : void =
        if(GameManager:=MaybeGameManager?, Player := player[Agent], Player.IsActive[],AgentBakery := GameManager.GetAgentBakery_Transacted(Agent)?):
            var UpdatedDataTable : player_data_table = AgentBakery.GameManager.QuestsAndOrdersDatabase.AddAgentQuest_ProgressV2(Player,AgentBakery.PlayerDataCache,1.0,FINDSCROLL_AREA0_QUEST_TYPE)(4)
            set UpdatedDataTable = AgentBakery.GameManager.QuestsAndOrdersDatabase.AddAgentQuest_ProgressV2(Player,UpdatedDataTable,1.0,FINDSCROLL_AREA2_QUEST_TYPE)(4)
            set UpdatedDataTable = AgentBakery.GameManager.QuestsAndOrdersDatabase.AddAgentQuest_ProgressV2(Player,UpdatedDataTable,1.0,FINDSCROLL_AREA8_QUEST_TYPE)(4)
            AgentBakery.SetNewPlayerDataCache(UpdatedDataTable)
            GameManager.DisplayCharacterMessage(Agent,W_CHARACTER,"This ancient Sourdough formula might just help...")



#STEVE GAMES
#SKILLED STEVE GAMES
    @editable:
        Categories := array{SteveGamesCategory}
    SteveSkilled_CoalReward : float = 10.0
    @editable:
        Categories := array{SteveGamesCategory} 
    AllSteve_PropComputers : []creative_prop = array{}
    @editable:
        Categories := array{SteveGamesCategory} 
    Steve_SkilledInteraction0 : skilled_interaction_device = skilled_interaction_device{}
    @editable:
        Categories := array{SteveGamesCategory}
    Steve_Button_Skilled0 : button_device = button_device{}
    @editable:
        Categories := array{SteveGamesCategory} 
    Steve_SkilledInteraction1 : skilled_interaction_device = skilled_interaction_device{}
    @editable:
        Categories := array{SteveGamesCategory}
    Steve_Button_Skilled1 : button_device = button_device{}
    @editable:
        Categories := array{SteveGamesCategory} 
    Steve_SkilledInteraction2 : skilled_interaction_device = skilled_interaction_device{}
    @editable:
        Categories := array{SteveGamesCategory}
    Steve_Button_Skilled2 : button_device = button_device{}

    OnBegin<override>()<suspends>:void=
        set AllSimonSoundAssignation = array{Do_AudioPlayer,Re_AudioPlayer,Mi_AudioPlayer,Fa_AudioPlayer,Sol_AudioPlayer}
        set SimonTeleporterOrigin = TeleporterSimonProp.GetTransform()

        for(Scroll : AllScrolls):
            set AllScrollsOriginalPositions = AllScrollsOriginalPositions + array{Scroll.GetTransform()}

    OnPlayerRemoved(AgentIndex : int) : void =
        if:
            set AllSimonPlayerLevel[AgentIndex] = 0
            set AllSimonPlayerReplies[AgentIndex] = array{}

    InitSubscribes() : void =
        TeleporterSimonArea.AgentEntersEvent.Subscribe(TryTeleportToSimonArea)
        for(SimonTeleporter : AllSimonTeleporters):
            SimonTeleporter.AgentEntersEvent.Subscribe(ExitFromSimon)

        
        for(Index:=0..AllArcheryButtons.Length-1, ArcheryButton:=AllArcheryButtons[Index]):
            ArcheryButton.InteractedWithEvent.Subscribe(handler_random{MaybeMinigamesManager:=option{Self},Index:=Index}.HandlerMinigamesManagerArcheryButton)

        #MANTI
        SkilledInteraction_Truck.InteractionSucceededEvent.Subscribe(handler_random{MaybeMinigamesManager:=option{Self},MiniGameQuestType:=TRUCKENGINE_MINIGAME_QUEST_TYPE, Index:=SKILLEDTRUCK_MINIGAME, Amount:=1.0}.HandlerSkilledGameCompleted)
        SkilledInteraction_Preassure.InteractionSucceededEvent.Subscribe(handler_random{MaybeMinigamesManager:=option{Self},MiniGameQuestType:=PREASSURESYSTEM_MINIGAME_QUEST_TYPE, Index:=SKILLEDPREASSURE_MINIGAME, Amount:=1.0}.HandlerSkilledGameCompleted)
        SkilledInteraction_Electrical.InteractionSucceededEvent.Subscribe(handler_random{MaybeMinigamesManager:=option{Self},MiniGameQuestType:=ELECTRICALSYSTEM_MINIGAME_QUEST_TYPE, Index:=SKILLEDELECTRICAL_MINIGAME, Amount:=1.0}.HandlerSkilledGameCompleted)

        Button_SkilledInteraction_Truck.InteractedWithEvent.Subscribe(handler_random{MaybeMinigamesManager:=option{Self}, Index:=SKILLEDTRUCK_MINIGAME}.HandlerDisplaySkilledGame)
        Button_SkilledInteraction_Preassure.InteractedWithEvent.Subscribe(handler_random{MaybeMinigamesManager:=option{Self}, Index:=SKILLEDPREASSURE_MINIGAME}.HandlerDisplaySkilledGame)
        Button_SkilledInteraction_Electrical.InteractedWithEvent.Subscribe(handler_random{MaybeMinigamesManager:=option{Self}, Index:=SKILLEDELECTRICAL_MINIGAME}.HandlerDisplaySkilledGame)

        #STEVE
        Steve_SkilledInteraction0.InteractionSucceededEvent.Subscribe(handler_random{MaybeMinigamesManager:=option{Self},MiniGameQuestType:=STEVEBAKER_QUEST_TYPE, Index:=STEVE_SKILLED0_MINIGAME, AuxIndex := 0, Amount:=1.0}.HandlerSkilledGameCompleted)
        Steve_SkilledInteraction1.InteractionSucceededEvent.Subscribe(handler_random{MaybeMinigamesManager:=option{Self},MiniGameQuestType:=STEVEBAKER_QUEST_TYPE, Index:=STEVE_SKILLED1_MINIGAME, AuxIndex := 1, Amount:=1.0}.HandlerSkilledGameCompleted)
        Steve_SkilledInteraction2.InteractionSucceededEvent.Subscribe(handler_random{MaybeMinigamesManager:=option{Self},MiniGameQuestType:=STEVEBAKER_QUEST_TYPE, Index:=STEVE_SKILLED2_MINIGAME, AuxIndex := 2, Amount:=1.0}.HandlerSkilledGameCompleted)
        
        Steve_Button_Skilled0.InteractedWithEvent.Subscribe(handler_random{MaybeMinigamesManager:=option{Self}, Index:=STEVE_SKILLED0_MINIGAME}.HandlerDisplaySkilledGame)
        Steve_Button_Skilled1.InteractedWithEvent.Subscribe(handler_random{MaybeMinigamesManager:=option{Self}, Index:=STEVE_SKILLED1_MINIGAME}.HandlerDisplaySkilledGame)
        Steve_Button_Skilled2.InteractedWithEvent.Subscribe(handler_random{MaybeMinigamesManager:=option{Self}, Index:=STEVE_SKILLED2_MINIGAME}.HandlerDisplaySkilledGame)

        #ARCHERY
        for(Index:=0..AllArcheryManipulatorDevices.Length-1, ArcheryManipulator:=AllArcheryManipulatorDevices[Index]):
            ArcheryManipulator.DamagedEvent.Subscribe(handler_random{MaybeMinigamesManager:=option{Self},Index:=Index}.HandlerArcheryDamagedGame)

        #SIMON BUTTONS
        for(PlayerIndex:=0..AllSimonButtons.Length-1, Button:=AllSimonButtons[PlayerIndex]):
            Button.InteractedWithEvent.Subscribe(HandlerAskSimon)

        #RED
        for(PlayerIndex:=0..AllSimonRedMutator.Length-1, Mutator:=AllSimonRedMutator[PlayerIndex]):
            Mutator.DamagedEvent.Subscribe(handler_random{MaybeMinigamesManager:=option{Self},Color:=SIMON_RED,Index:=PlayerIndex}.HandlerSimonManipulator)
        #BLUE
        for(PlayerIndex:=0..AllSimonBlueMutator.Length-1, Mutator:=AllSimonBlueMutator[PlayerIndex]):
            Mutator.DamagedEvent.Subscribe(handler_random{MaybeMinigamesManager:=option{Self},Color:=SIMON_BLUE,Index:=PlayerIndex}.HandlerSimonManipulator)
        #GREEN
        for(PlayerIndex:=0..AllSimonGreenMutator.Length-1, Mutator:=AllSimonGreenMutator[PlayerIndex]):
            Mutator.DamagedEvent.Subscribe(handler_random{MaybeMinigamesManager:=option{Self},Color:=SIMON_GREEN,Index:=PlayerIndex}.HandlerSimonManipulator)
        #YELLOW
        for(PlayerIndex:=0..AllSimonYellowMutator.Length-1, Mutator:=AllSimonYellowMutator[PlayerIndex]):
            Mutator.DamagedEvent.Subscribe(handler_random{MaybeMinigamesManager:=option{Self},Color:=SIMON_YELLOW,Index:=PlayerIndex}.HandlerSimonManipulator)
        #WHITE
        for(PlayerIndex:=0..AllSimonWhiteMutator.Length-1, Mutator:=AllSimonWhiteMutator[PlayerIndex]):
            Mutator.DamagedEvent.Subscribe(handler_random{MaybeMinigamesManager:=option{Self},Color:=SIMON_WHITE,Index:=PlayerIndex}.HandlerSimonManipulator)

        #SCROLL FIND
        for(Scroll : AllScrolls):
            Scroll.CollectedEvent.Subscribe(ScrollCollectedEvent)
         
    MiniGamesTick() : void =
        SimonTick()
        HBInestabilityTick()


    Init(GameManager : game_manager) : void =
        set MaybeGameManager = option{GameManager}
        InitSubscribes()

        #Arrow game
            set AllArcheryDamagingAgents = array{AllArcheryDamagingAgents0,AllArcheryDamagingAgents1,AllArcheryDamagingAgents2,AllArcheryDamagingAgents3,AllArcheryDamagingAgents4,AllArcheryDamagingAgents5,AllArcheryDamagingAgents6,AllArcheryDamagingAgents7}
            set AllPillars = array{AllPillars0,AllPillars1,AllPillars2,AllPillars3,AllPillars4,AllPillars5,AllPillars6,AllPillars7}

            for(Pillars : AllPillars):
                var AllTransforms : []transform = array{}
                for(Pillar : Pillars):
                    OriginalTransform := Pillar.GetTransform()
                    set AllTransforms = AllTransforms + array{OriginalTransform}
                set AllPillarsOriginalPositions = AllPillarsOriginalPositions + array{AllTransforms}

            var AllDamages : []float = array{
                CalculeWeaponStat(WEAPONSTAT_TYPE_DAMAGE,BOW_TYPE,1,1)(0)*10,
                CalculeWeaponStat(WEAPONSTAT_TYPE_DAMAGE,BOW_TYPE,1,3)(0)*10,
                CalculeWeaponStat(WEAPONSTAT_TYPE_DAMAGE,BOW_TYPE,1,7)(0)*10,
                CalculeWeaponStat(WEAPONSTAT_TYPE_DAMAGE,BOW_TYPE,1,10)(0)*10,
                CalculeWeaponStat(WEAPONSTAT_TYPE_DAMAGE,BOW_TYPE,2,1)(0)*10,
                CalculeWeaponStat(WEAPONSTAT_TYPE_DAMAGE,BOW_TYPE,2,3)(0)*10,
                CalculeWeaponStat(WEAPONSTAT_TYPE_DAMAGE,BOW_TYPE,2,7)(0)*10,
                CalculeWeaponStat(WEAPONSTAT_TYPE_DAMAGE,BOW_TYPE,2,10)(0)*10,
                CalculeWeaponStat(WEAPONSTAT_TYPE_DAMAGE,BOW_TYPE,3,1)(0)*10,
                CalculeWeaponStat(WEAPONSTAT_TYPE_DAMAGE,BOW_TYPE,3,3)(0)*10,
                CalculeWeaponStat(WEAPONSTAT_TYPE_DAMAGE,BOW_TYPE,3,7)(0)*10,
                CalculeWeaponStat(WEAPONSTAT_TYPE_DAMAGE,BOW_TYPE,3,10)(0)*10}

            set AllArcheryMiniGame_MaxHealth = SortFloatArray(AllDamages)

            for(Index:=0..AllArcheryManipulatorDevices.Length-1, ArrowManipulator := AllArcheryManipulatorDevices[Index], MaxHealth := AllArcheryMiniGame_MaxHealth[Index], MaxTime := AllArcheryPillarsMaxTime[Index]):
                set AllArcheryMiniGame_RegenerationActive = AllArcheryMiniGame_RegenerationActive + array{false}
                set AllArcheryMiniGame_CurrentHealth = AllArcheryMiniGame_CurrentHealth + array{MaxHealth}
                set AllArcheryState = AllArcheryState + array{MaxTime}

            spawn:
                DelayResetPillars()
            
    MinigameReward(Agent : agent, Bakery : bakery_machines_manager, IndexMiniGame : int, Reward : float, RewardType : int) : void =
        var NewData : player_data_table = Bakery.PlayerDataCache
        #RandomCoolDown := GetRandomInt(300,600)
        if(GameManager:=MaybeGameManager?):
            GameManager.VFXManager.BigObjetive_AudioPlayer.Play(Agent)
            set NewData = GameManager.PlayersDataManager.AddMiniGameCooldown(NewData,IndexMiniGame,MiniGameCoolDown)
            set NewData = GameManager.PlayersDataManager.AddResourcers(NewData,Reward,RewardType,-1.0)(0)
            GameManager.DisplayResourcersGather(Agent,array{Reward},array{COAL},false)
            Bakery.SetNewPlayerDataCache(NewData)
            if(IndexMiniGame>2):
                spawn:
                    DisplayResultMessage(Agent,MiniGameCoolDown,J_CHARACTER)
            else:
                spawn:
                    DisplayResultMessage(Agent,MiniGameCoolDown,W_CHARACTER)

    CalculeCoalReward(IsSimon : logic, Agent : agent, ?ArcheryIndex : int = 0) : float = 
        var Coal : float = 0.0
        if(GameManager:=MaybeGameManager?, AgentBakery := GameManager.GetAgentBakery_Transacted(Agent)?):
            if(IsSimon?):
                NumAreaReq_PerLevel := (AgentBakery.PlayerDataCache.LV_Area*1.0)/(AllArcheryManipulatorDevices.Length*1.0)
                if(IntReward := Ceil[Archery_CoalReward_PerLevel*NumAreaReq_PerLevel*3.0]):
                    set Coal = (IntReward+10)*1.0
            else:
                set Coal = Archery_CoalReward_PerLevel*(ArcheryIndex+1)
        return Coal

#SIMON GAMES
    SimonTick() : void =
        if(GameManager:=MaybeGameManager?):
            #if(GameManager.DebugOn?, SimonCoolDown >= 60):
            #    set SimonCoolDown = SimonMaxCoolDown
            if(SimonGameActive = -1, SimonCoolDown < SimonMaxCoolDown):
                set SimonCoolDown = SimonCoolDown + 1
            else if(SimonGameActive = -1, SimonCoolDown >= SimonMaxCoolDown):
                StartSimonGame()
                    for(Player : GetPlayspace().GetPlayers(), Player.IsActive[], Agent := agent[Player], AgentBakery := GameManager.GetAgentBakery_Transacted(Agent)?, AgentBakery.PlayerDataCache.Inventory_Fame>50.0):
                        GameManager.DisplayCharacterMessage(Agent,G_CHARACTER,"The analysis system teleporter has appeared in the city!")  
            else:
                if(SimonMaxTime<=SimonTotalTime):
                    Log("Simon time ended!",LOG_LEVEL_MINIGAMES) 
                    for(Player : GetPlayspace().GetPlayers(), Player.IsActive[], Agent := agent[Player], AgentBakery := GameManager.GetAgentBakery_Transacted(Agent)?, AgentBakery.PlayerDataCache.Inventory_Fame>50.0):
                        DungeonPlaying:=GameManager.GeneralMobsManager.CheckIfIsPlaying(Agent)
                        if(DungeonPlaying=-1, ActionInfoUI := AgentBakery.MaybeActionInfoUI?):
                            ActionInfoUI.RemoveSimonInfo()
                    StopSimonGame()
                else: 
                    for(Player : GetPlayspace().GetPlayers(), Player.IsActive[], Agent := agent[Player], AgentBakery := GameManager.GetAgentBakery_Transacted(Agent)?, AgentBakery.PlayerDataCache.Inventory_Fame>50.0, ActionInfoUI := AgentBakery.MaybeActionInfoUI?):
                        AgentIndex:=GameManager.GetAgentIndex(Agent)
                        DungeonPlaying:=GameManager.GeneralMobsManager.CheckIfIsPlaying(Agent)
                        MaybeAgentHexaManager:=GameManager.GetAgentHexaManager(Agent)
                        if(AgentHexaManager:=MaybeAgentHexaManager?):
                            SimonUICanBeDisplayed:=AgentHexaManager.SimonUICanBeDisplayed()
                            if(DungeonPlaying=-1,SimonUICanBeDisplayed?,Level:=AllSimonPlayerLevel[AgentIndex], Level<=AllSimonSays.Length-1):
                                TupleTime := TransformToMinutes_Seconds(SimonMaxTime-SimonTotalTime)
                                ActionInfoUI.Set_SimonInfo("Analyzer ready: {TupleTime(0)}:{TupleTime(1)}")
                                #Log("Simon time: {SimonTotalTime} of {SimonMaxTime}",LOG_LEVEL_MINIGAMES) 
                            else:
                                ActionInfoUI.RemoveSimonInfo()
                                #Log("Error Player is on Dungeon {DungeonPlaying}",LOG_LEVEL_MINIGAMES) 
                    set SimonTotalTime = SimonTotalTime + 1

    SimonPressed(Agent : agent, PlayerIndex : int, Color : int) : void =
        if(GameManager:=MaybeGameManager?, AgentBakery := GameManager.GetAgentBakery_Transacted(Agent)?, SimonIsAlreadyTalking:=AllSimonIsTalking[PlayerIndex], not SimonIsAlreadyTalking?):
            if(SimonGameActive=SIMON_COLOR):
                if(Color=SIMON_RED, RedProp := AllSimonRed[PlayerIndex]):
                    spawn:
                        PlayColorChange(Agent, RedProp, SIMON_RED)
                else if(Color=SIMON_GREEN, GreenProp := AllSimonGreen[PlayerIndex]):
                    spawn:
                        PlayColorChange(Agent, GreenProp, SIMON_GREEN)
                else if(Color=SIMON_BLUE, BlueProp := AllSimonBlue[PlayerIndex]):
                    spawn:
                        PlayColorChange(Agent, BlueProp, SIMON_BLUE)
                else if(Color=SIMON_YELLOW, YellowProp := AllSimonYellow[PlayerIndex]):
                    spawn:
                        PlayColorChange(Agent, YellowProp, SIMON_YELLOW)
                else if(Color=SIMON_WHITE, WhiteProp := AllSimonWhite[PlayerIndex]):
                    spawn:
                        PlayColorChange(Agent, WhiteProp, SIMON_WHITE)
            else if(SimonGameActive=SIMON_SOUND):
                if(Color=SIMON_RED, RedProp := AllSimonRed[PlayerIndex]):
                    spawn:
                        PlayColorChange(Agent, RedProp, SIMON_RED)
                else if(Color=SIMON_GREEN, GreenProp := AllSimonGreen[PlayerIndex]):
                    spawn:
                        PlayColorChange(Agent, GreenProp, SIMON_GREEN)
                else if(Color=SIMON_BLUE, BlueProp := AllSimonBlue[PlayerIndex]):
                    spawn:
                        PlayColorChange(Agent, BlueProp, SIMON_BLUE)
                else if(Color=SIMON_YELLOW, YellowProp := AllSimonYellow[PlayerIndex]):
                    spawn:
                        PlayColorChange(Agent, YellowProp, SIMON_YELLOW)
                else if(Color=SIMON_WHITE, WhiteProp := AllSimonWhite[PlayerIndex]):
                    spawn:
                        PlayColorChange(Agent, WhiteProp, SIMON_WHITE)

            if(var PlayerCurrentReply : []int =AllSimonPlayerReplies[PlayerIndex]):
                set PlayerCurrentReply = PlayerCurrentReply + array{Color}
                if(set AllSimonPlayerReplies[PlayerIndex] = PlayerCurrentReply):

            #if(SimonGameActive=SIMON_COLOR):
            PlayerReplyState:=CheckSimonPlayerReply(Agent,PlayerIndex)
            if(PlayerReplyState=1):
                if(OldLevel:=AllSimonPlayerLevel[PlayerIndex], NewLevel := OldLevel+1, set AllSimonPlayerLevel[PlayerIndex]=NewLevel):
                    if(set AllSimonPlayerReplies[PlayerIndex] = array{}): #Cleaning player reply
                    if(NewLevel>AllSimonSays.Length-1):
                        #spawn:
                        #    GameManager.VFXManager.PlaySound(SMALLOBJETIVE_SOUND,1.0,option{Agent},false)
                        Log("Simon complete, granting rewards",LOG_LEVEL_MINIGAMES)
                        Reward:=CalculeCoalReward(true,Agent)
                        AgentBakery.DisplayResourcersGather(array{Reward},array{COAL},false)
                        AgentBakery.SaveResourcers(Agent,array{COAL},array{Reward})
                        AgentBakery.RefreshAllUI()
                        ExitFromSimon(Agent)
                        GameManager.DisplayCharacterMessage(Agent, I_CHARACTER, "Stale Bread Analysis completed!",?DisplayTime:=5.0, ?Sound := BIGOBJETIVE_SOUND)
                        GameManager.DisplayCharacterMessage(Agent, G_CHARACTER, "Analysis is more rewarding with more areas unlocked", ?DisplayTime:=5.0)
                        #MUST RESET PLAYER REPLY TO AVOID DOUBLE 
                    else:
                        GameManager.DisplayCharacterMessage(Agent, I_CHARACTER, "Completed {NewLevel}/{AllSimonSays.Length}", ?Sound := SMALLOBJETIVE_SOUND)
                        Log("Simon level complete, showing next level",LOG_LEVEL_MINIGAMES)
                        #spawn:
                        #    GameManager.VFXManager.PlaySound(BIGOBJETIVE_SOUND,1.0,option{Agent},false)
                        spawn:
                            AskSimon(Agent,2.0)
            else if(PlayerReplyState=0):
                Log("Simon incomplete, doing nothing",LOG_LEVEL_MINIGAMES)
            else if(PlayerReplyState=-1):
                Log("Simon wrong reseting player orders",LOG_LEVEL_MINIGAMES)
                if(set AllSimonPlayerReplies[PlayerIndex] = array{}):
                    spawn:
                        GameManager.VFXManager.PlaySound(ERROR_SOUND,0.5,option{Agent},false)

    CheckSimonPlayerReply(Agent : agent, Index : int) : int =
        var PlayerReplyState : int = 0
        if(PlayerCurrentReply:=AllSimonPlayerReplies[Index]):
            if(PlayerLevel := AllSimonPlayerLevel[Index], SimonOrders:= AllSimonSays[PlayerLevel]):
                var OrderPosition : int = 0 
                loop:
                    if(SimonOrder:=SimonOrders[OrderPosition], ReplyOrder := PlayerCurrentReply[OrderPosition]):
                        if(SimonOrder = ReplyOrder):
                            #Log("Simon order completed",LOG_LEVEL_MINIGAMES)
                            if(OrderPosition<SimonOrders.Length-1):
                                set OrderPosition = OrderPosition + 1
                            else:
                                Log("All Simon orders completed",LOG_LEVEL_MINIGAMES)
                                set PlayerReplyState = 1
                                break
                        else:
                            set PlayerReplyState = -1
                            Log("Simon order wrong. Order was {SimonOrder}, player pressed {ReplyOrder}",LOG_LEVEL_MINIGAMES)
                            break
                    else:
                        Log("Reply order not found on OrderPosition: {OrderPosition} breaking",LOG_LEVEL_MINIGAMES)
                        set PlayerReplyState = 0
                        break
        return PlayerReplyState

    ExitFromSimon(Agent : agent) : void =
        if(GameManager:=MaybeGameManager?, ValidBakery:=GameManager.GetAgentBakery_Transacted(Agent)?):
            <#
            if(SimonGameActive>-1):
                GameManager.TeleporterManager.TeleportPlayerUsingTeleporter(Agent,TeleporterSimonArea.GetTransform())
            else:#>
            GameManager.TeleporterManager.TryToFreeTeleportToBakery(Agent)
            ValidBakery.CheckMusicPlay(BAKERSMUSIC_SOUND)


    SpawnSimonTeleporter<private>() : void =
        RandomIndex:=GetRandomInt(0,AllPosibleSimonTeleporter_Positions.Length-1)
        if(Position:=AllPosibleSimonTeleporter_Positions[RandomIndex]):
            ObjetiveTransform:=AdaptTransformFromVisualEmpty(Position.GetTransform(),VisualEmptyHideZ)
            if(TeleporterSimonArea.TeleportTo[ObjetiveTransform]):
            else:
                TeleportFailedAttempt(false, false, option{TeleporterSimonProp}, ObjetiveTransform, 128.0)
            TeleporterSimonArea.Enable()

    DespawnSimonTeleporter<private>() : void =
        if(TeleporterSimonArea.TeleportTo[SimonTeleporterOrigin]):
        else:
            TeleportFailedAttempt(false, false, option{TeleporterSimonProp}, SimonTeleporterOrigin, 128.0)
        TeleporterSimonArea.Disable()

    SpawnSimonTimeOld_<private>()<suspends> : void =
        if(GameManager:=MaybeGameManager?):
            for(Player : GetPlayspace().GetPlayers(), Player.IsActive[], Agent := agent[Player], AgentBakery := GameManager.GetAgentBakery_Transacted(Agent)?):
                GameManager.DisplayCharacterMessage(Agent,G_CHARACTER,"The analysis system teleporter has appeared in the city!",?Sound:=STARTCHALLENGE_SOUND)
            loop:
                if(SimonMaxTime<=SimonTotalTime):
                    Log("Simon time ended!",LOG_LEVEL_MINIGAMES) 
                    for(Player : GetPlayspace().GetPlayers(), Player.IsActive[], Agent := agent[Player], AgentBakery := GameManager.GetAgentBakery_Transacted(Agent)?):
                        DungeonPlaying:=GameManager.GeneralMobsManager.CheckIfIsPlaying(Agent)
                        if(DungeonPlaying=-1, ActionInfoUI := AgentBakery.MaybeActionInfoUI?):
                            ActionInfoUI.RemoveSimonInfo()
                    StopSimonGame()
                    break
                else: 
                    for(Player : GetPlayspace().GetPlayers(), Player.IsActive[], Agent := agent[Player], AgentBakery := GameManager.GetAgentBakery_Transacted(Agent)?, ActionInfoUI := AgentBakery.MaybeActionInfoUI?):
                        DungeonPlaying:=GameManager.GeneralMobsManager.CheckIfIsPlaying(Agent)

                        MaybeAgentHexaManager:=GameManager.GetAgentHexaManager(Agent)
                        if(AgentHexaManager:=MaybeAgentHexaManager?):
                            SimonUICanBeDisplayed:=AgentHexaManager.SimonUICanBeDisplayed()
                            if(DungeonPlaying=-1,SimonUICanBeDisplayed?):
                                TupleTime := TransformToMinutes_Seconds(SimonMaxTime-SimonTotalTime)
                                ActionInfoUI.Set_SimonInfo("Analyzer ready: {TupleTime(0)}:{TupleTime(1)}")
                                #Log("Simon time: {SimonTotalTime} of {SimonMaxTime}",LOG_LEVEL_MINIGAMES) 
                            else:
                                ActionInfoUI.RemoveSimonInfo()
                                #Log("Error Player is on Dungeon {DungeonPlaying}",LOG_LEVEL_MINIGAMES) 
                    set SimonTotalTime = SimonTotalTime + 1
                Sleep(1.0)
    StopSimonGame() : void =
        set SimonCoolDown = 0
        set SimonGameActive = -1
        DespawnSimonTeleporter()
        set AllSimonPlayerLevel = array{0,0,0,0}
        set AllSimonPlayerReplies = array{array{},array{},array{},array{}}
        if(GameManager:=MaybeGameManager?):
            for(Player : GetPlayspace().GetPlayers(), Agent := agent[Player],FortCharacter := Player.GetFortCharacter[],ValidBakery:=GameManager.GetAgentBakery_Transacted(Agent)?):
                for(Teleporter : AllSimonTeleporters):
                    if(4096.0>Distance(Teleporter.GetTransform().Translation, FortCharacter.GetTransform().Translation)):
                        GameManager.TeleporterManager.TryToFreeTeleportToBakery(Agent)
                        GameManager.DisplayCharacterMessage(Agent,I_CHARACTER,"The time has ended!",?Sound:=ERROR_SOUND)
                        ValidBakery.CheckMusicPlay(BAKERSMUSIC_SOUND)

    StartSimonGame() : void = 
        #spawn:
        #    SpawnSimonTime()
        SpawnSimonTeleporter()
        set SimonTotalTime = 0
        set AllSimonPlayerLevel = array{0,0,0,0}
        set AllSimonPlayerReplies = array{array{},array{},array{},array{}}
        #set SimonGameActive = GetRandomInt(SIMON_COLOR,SIMON_SOUND)
        set SimonGameActive = SIMON_COLOR
        set AllSimonSoundAssignation = Shuffle(AllSimonSoundAssignation)
        set AllSimonIsTalking = array{false,false,false,false}
        #Generate SimonSays
        set AllSimonSays = array{} #Clear Simon Says
        if(SimonGameActive=SIMON_COLOR):
            for(Index0:=0..3): #4 Levels
                var ArrayMovements : []int = array{}
                for(Index1:=0..2+Index0): #3 Movements
                    RandomOption:=GetRandomInt(0,4)
                    set ArrayMovements = ArrayMovements + array{RandomOption}
                set AllSimonSays = AllSimonSays + array{ArrayMovements}
        else if(SimonGameActive=SIMON_SOUND):
            for(Index0:=0..2): #3 Levels
                var ArrayMovements : []int = array{}
                for(Index1:=0..1+Index0): #2 Movements
                    RandomOption:=GetRandomInt(0,4)
                    set ArrayMovements = ArrayMovements + array{RandomOption}
                set AllSimonSays = AllSimonSays + array{ArrayMovements}
        else:
            PrintFatalError("SimonGameActive: {SimonGameActive} INVALID!")
        SpawnSimonTeleporter()

    HandlerAskSimon(Agent : agent) : void =
        if(GameManager:=MaybeGameManager?):
            AgentIndex:=GameManager.GetAgentIndex(Agent)
            if(SimonIsAlreadyTalking:=AllSimonIsTalking[AgentIndex], not SimonIsAlreadyTalking?, set AllSimonIsTalking[AgentIndex] = true):
                Log("Asking Simon button {AgentIndex}",LOG_LEVEL_MINIGAMES)
                spawn:
                    AskSimon(Agent,0.0)
        else:
            PrintFatalError("Game Manager not found")

    AskSimon(Agent : agent, PreDelay : float)<suspends> : void =
        Sleep(PreDelay)
        DelayBetweenOrders : float = 0.5
        if(GameManager:=MaybeGameManager?):
            AgentIndex:=GameManager.GetAgentIndex(Agent)
            if(PlayerLevel:=AllSimonPlayerLevel[AgentIndex], SimonOrders := AllSimonSays[PlayerLevel]):
                if(SimonGameActive=SIMON_COLOR):
                    Log("Simon Says colors of Level: {PlayerLevel}",LOG_LEVEL_MINIGAMES)
                    for(SimonColor : SimonOrders):
                        if(SimonColor=SIMON_RED, RedProp := AllSimonRed[AgentIndex]):
                            PlayColorChange(Agent, RedProp, SIMON_RED)
                        else if(SimonColor=SIMON_GREEN, GreenProp := AllSimonGreen[AgentIndex]):
                            PlayColorChange(Agent, GreenProp, SIMON_GREEN)
                        else if(SimonColor=SIMON_BLUE, BlueProp := AllSimonBlue[AgentIndex]):
                            PlayColorChange(Agent, BlueProp, SIMON_BLUE)
                        else if(SimonColor=SIMON_YELLOW, YellowProp := AllSimonYellow[AgentIndex]):
                            PlayColorChange(Agent, YellowProp, SIMON_YELLOW)
                        else if(SimonColor=SIMON_WHITE, WhiteProp := AllSimonWhite[AgentIndex]):
                            PlayColorChange(Agent, WhiteProp, SIMON_WHITE)
                        Sleep(DelayBetweenOrders)
                else if(SimonGameActive=SIMON_SOUND):  
                    Log("Simon Says sounds of Level: {PlayerLevel}",LOG_LEVEL_MINIGAMES) 
                    for(SimonSound : SimonOrders):   
                        if(SimonSound=SIMON_RED, RedProp := AllSimonRed[AgentIndex]):
                            PlaySimonSound(Agent, RedProp, SIMON_RED)
                        else if(SimonSound=SIMON_GREEN, GreenProp := AllSimonGreen[AgentIndex]):
                            PlaySimonSound(Agent, GreenProp, SIMON_GREEN)
                        else if(SimonSound=SIMON_BLUE, BlueProp := AllSimonBlue[AgentIndex]):
                            PlaySimonSound(Agent, BlueProp, SIMON_BLUE)
                        else if(SimonSound=SIMON_YELLOW, YellowProp := AllSimonYellow[AgentIndex]):
                            PlaySimonSound(Agent, YellowProp, SIMON_YELLOW)
                        else if(SimonSound=SIMON_WHITE, WhiteProp := AllSimonWhite[AgentIndex]):
                            PlaySimonSound(Agent, WhiteProp, SIMON_WHITE)
                        Sleep(DelayBetweenOrders)  
            if(set AllSimonIsTalking[AgentIndex] = false):
 
    PlaySimonSound(Agent : agent, Prop : creative_prop, Color : int)<suspends> : void =
        DelayColorChange : float = 0.5
        if(AudioPlayer := AllSimonSoundAssignation[Color]):
            AudioPlayer.Play(Agent)

    PlayColorChange(Agent : agent, Prop : creative_prop, Color : int)<suspends> : void =
        DelayColorChange : float = 0.5
        if(AudioPlayer := AllSimonSoundAssignation[Color]):
            AudioPlayer.Play(Agent)
        if(Color=SIMON_RED):
            Prop.SetMaterial(Materials.Mat_HB_PimentosWeakHigh)
            Sleep(DelayColorChange)
            Prop.SetMaterial(Materials.Mat_MasterHardBread_Pimentos)
        else if(Color=SIMON_GREEN):
            Prop.SetMaterial(Materials.Mat_HB_MutantWeakHigh)
            Sleep(DelayColorChange)
            Prop.SetMaterial(Materials.Mat_MasterHardBread_Mutant)
        else if(Color=SIMON_BLUE):
            Prop.SetMaterial(Materials.Mat_HB_WaterWeakHigh)
            Sleep(DelayColorChange)
            Prop.SetMaterial(Materials.Mat_MasterHardBread_Water)
        else if(Color=SIMON_WHITE):
            Prop.SetMaterial(Materials.Mat_HB_FlourWeakHigh)
            Sleep(DelayColorChange)
            Prop.SetMaterial(Materials.Mat_MasterHardBread_Flour)
        else if(Color=SIMON_YELLOW):
            Prop.SetMaterial(Materials.Mat_HB_CheeseWeakHigh)
            Sleep(DelayColorChange)
            Prop.SetMaterial(Materials.Mat_MasterHardBread_Cheese)

                
    TryTeleportToSimonArea(Agent : agent) : void =
        if(SimonGameActive>-1, GameManager:=MaybeGameManager?, Player := player[Agent], Character := Player.GetFortCharacter[], Player.IsActive[]):
            GameManager.VFXManager.StopAllMusics(Agent)
            AgentIndex:=GameManager.GetAgentIndex(Agent)
            if(Level:=AllSimonPlayerLevel[AgentIndex], Level<=AllSimonSays.Length-1):
                if(SimonGameActive=SIMON_COLOR):
                    GameManager.DisplayCharacterMessage(Agent, I_CHARACTER, "Repeat the color pattern!", ?Sound := MESSAGE_SOUND)
                else if(SimonGameActive=SIMON_SOUND):
                    GameManager.DisplayCharacterMessage(Agent, I_CHARACTER, "Repeat the song you just heard!", ?Sound := MESSAGE_SOUND)
                if(Position:=AllSimon_Positions[AgentIndex]):
                    GameManager.TeleporterManager.TeleportPlayerUsingTeleporter(Agent,AdaptTransformFromVisualEmpty(Position.GetTransform(),VisualEmptyHideZ))
            else:
                GameManager.DisplayCharacterMessage(Agent, I_CHARACTER, "You have already completed the Stale bread analyzer", ?Sound:=ERROR_SOUND)

    PlaySimon(Agent : agent) : void =
        if(GameManager:=MaybeGameManager?):
            GameManager.GetAgentIndex(Agent)
            if(SimonGameActive=SIMON_COLOR):
            else if(SimonGameActive=SIMON_SOUND):


#SKILLED INTERACTION DEVICES 
    TryAwardSkilledMiniGame(Agent : agent, MINIGAME_QUEST_TYPE : int, MiniGameIndex : int, AuxIndex : int, ProgressAmount : float) : void =
        if(GameManager:=MaybeGameManager?, ValidBakery := GameManager.GetAgentBakery_Transacted(Agent)?):
            ValidBakery.CheckQuestProgress(ProgressAmount,MINIGAME_QUEST_TYPE)
            if(CoolDown:=ValidBakery.PlayerDataCache.AllMiniGamesCooldowns[MiniGameIndex]):
                if(CoolDown<=0):
                    if(MiniGameIndex>2):
                        MinigameReward(Agent,ValidBakery,MiniGameIndex,SteveSkilled_CoalReward,COAL)
                        spawn:
                            SteveComputerCompletedEventCeremony(AuxIndex)
                    else:
                        MinigameReward(Agent,ValidBakery,MiniGameIndex,Skilled_CoalReward,COAL)
                else:
                    if(MiniGameIndex>2):
                        spawn:
                            DisplayResultMessage(Agent,CoolDown,G_CHARACTER)
                    else:
                        spawn:
                            DisplayResultMessage(Agent,CoolDown,G_CHARACTER)


        <#
        TryAwardSkilledPulsingMiniGame(Agent : agent) : void =
            if(GameManager:=MaybeGameManager?, ValidBakery := GameManager.GetAgentBakery_Transacted(Agent)?):
                ValidBakery.CheckQuestProgress(1.0,SKILLEDPULSING_MINIGAME_QUEST_TYPE)
                if(CoolDown:=ValidBakery.PlayerDataCache.AllMiniGamesCooldowns[SKILLEDPULSING0_MINIGAME]):
                    if(CoolDown<=0):
                        MinigameReward(Agent,ValidBakery,SKILLEDPULSING0_MINIGAME)
                    else:
                        spawn:
                            DisplayResultMessage(Agent,CoolDown)

        TryAwardSkilledBarMiniGame(Agent : agent) : void =
            if(GameManager:=MaybeGameManager?, ValidBakery := GameManager.GetAgentBakery_Transacted(Agent)?):
                ValidBakery.CheckQuestProgress(1.0,TRUCKENGINE_MINIGAME_QUEST_TYPE)
                if(CoolDown:=ValidBakery.PlayerDataCache.AllMiniGamesCooldowns[SKILLEDBAR0_MINIGAME]):
                    if(CoolDown<=0):
                        MinigameReward(Agent,ValidBakery,SKILLEDBAR0_MINIGAME)
                    else:
                        spawn:
                            DisplayResultMessage(Agent,CoolDown)
                            

        TryAwardSkilledCircularMiniGame(Agent : agent) : void =
            if(GameManager:=MaybeGameManager?, ValidBakery := GameManager.GetAgentBakery_Transacted(Agent)?):
                ValidBakery.CheckQuestProgress(1.0,SKILLEDCIRCULAR_MINIGAME_QUEST_TYPE)
                if(CoolDown:=ValidBakery.PlayerDataCache.AllMiniGamesCooldowns[SKILLEDCIRCULAR0_MINIGAME]): 
                    if(CoolDown<=0):
                        MinigameReward(Agent,ValidBakery,SKILLEDCIRCULAR0_MINIGAME)
                    else:
                        spawn:
                            DisplayResultMessage(Agent,CoolDown)#>

    DisplaySkilledMiniGame(Agent : agent, IndexGame : int) : void =
        if(IndexGame=SKILLEDTRUCK_MINIGAME):
            if(SkilledInteraction_Truck.BeginInteraction[Agent]):
        else if(IndexGame=SKILLEDELECTRICAL_MINIGAME):
            if(SkilledInteraction_Electrical.BeginInteraction[Agent]):
        else if(IndexGame=SKILLEDPREASSURE_MINIGAME):    
            if(SkilledInteraction_Preassure.BeginInteraction[Agent]):
        else if(IndexGame=STEVE_SKILLED0_MINIGAME):    
            if(Steve_SkilledInteraction0.BeginInteraction[Agent]):    
        else if(IndexGame=STEVE_SKILLED1_MINIGAME):   
            if(Steve_SkilledInteraction1.BeginInteraction[Agent]):  
        else if(IndexGame=STEVE_SKILLED2_MINIGAME):   
            if(Steve_SkilledInteraction2.BeginInteraction[Agent]): 

    DisplayResultMessage(Agent : agent, CoolDown : int, Character : int)<suspends> : void =
        Sleep(1.0)
        if(GameManager:=MaybeGameManager?):
            TupleTime := TransformToMinutes_Seconds(CoolDown)
            if(CoolDown<=0):
                #GameManager.DisplayCharacterMessage(Agent,I_CHARACTER,"Success! You can claim this reward again in {TupleTime(0)}:{TupleTime(1)}", ?Sound := SMALLOBJETIVE_SOUND)
            else:
                GameManager.DisplayCharacterMessage(Agent,I_CHARACTER,"You can claim this reward again in {TupleTime(0)}:{TupleTime(1)}", ?Sound := SMALLOBJETIVE_SOUND)

#ARROW MINIGAME
    ArcheryButtonTriggered(TriggerAgent : agent, Index: int) : void =
        TryAwardArcheryMiniGame(TriggerAgent,Index)
        Log("Damaged event", LOG_LEVEL_MINIGAMES)
        if(var AllDamagingAgents : [agent]int = AllArcheryDamagingAgents[Index]):
            for(Player : GetPlayspace().GetPlayers(), OtherAgent:=agent[Player], DamageDelay:=AllDamagingAgents[OtherAgent]):
                TryAwardArcheryMiniGame(OtherAgent,Index)

    TryAwardArcheryMiniGame(Agent : agent, ArcheryIndex : int) : void =
        AllArcheryIndexes:=array{ARCHERY0_MINIGAME,ARCHERY1_MINIGAME,ARCHERY2_MINIGAME,ARCHERY3_MINIGAME,ARCHERY4_MINIGAME,ARCHERY5_MINIGAME,ARCHERY6_MINIGAME,ARCHERY7_MINIGAME,ARCHERY8_MINIGAME}
        if(GameManager:=MaybeGameManager?, ValidBakery := GameManager.GetAgentBakery_Transacted(Agent)?):
            ValidBakery.CheckQuestProgress(1.0,ARCHERY_MINIGAME_QUEST_TYPE)
            if(MiniGameIndex:=AllArcheryIndexes[ArcheryIndex],CoolDown:=ValidBakery.PlayerDataCache.AllMiniGamesCooldowns[MiniGameIndex]): 
                spawn:
                    DisplayResultMessage(Agent,CoolDown,G_CHARACTER)
                if(CoolDown<=0):
                    Reward:=CalculeCoalReward(false,Agent,?ArcheryIndex:=ArcheryIndex)
                    MinigameReward(Agent,ValidBakery,MiniGameIndex,Reward,COAL)
            else:
                PrintFatalError("OnTryAwardArcheryMiniGame ArcheryIndex: {ArcheryIndex} INVALID!")

    DelayResetPillars()<suspends> : void =
        Sleep(5.0)
        for(Index:=0..AllPillars.Length-1, PillarsSelected:=AllPillars[Index]):
            for(IndexPillar:=0..PillarsSelected.Length-1, Pillar := PillarsSelected[IndexPillar], OriginalPosition:=AllPillarsOriginalPositions[Index][IndexPillar]):
                ObjetiveTransform : transform =transform{Translation:=vector3{X:=OriginalPosition.Translation.X,Y:=OriginalPosition.Translation.Y,Z:=OriginalPosition.Translation.Z-ArcheryPillarsTranslation}, Rotation:=OriginalPosition.Rotation, Scale:=OriginalPosition.Scale}
                spawn:
                    Pillar.MoveTo(ObjetiveTransform,0.5)

    
    ArcheryDamagedEvent(Agent : agent, Index : int) : void =
        Log("Damaged event", LOG_LEVEL_MINIGAMES)
        if(GameManager:=MaybeGameManager?, State := AllArcheryState[Index], MaxTime := AllArcheryPillarsMaxTime[Index], State=MaxTime):
            TryRunArcheryRegeneration(Index)
            if(Manipulator := AllArcheryManipulatorDevices[Index], MaxHealth := AllArcheryMiniGame_MaxHealth[Index], var CurrentHealth : float = AllArcheryMiniGame_CurrentHealth[Index]):
                TupleDamage:=GameManager.GeneralMobsManager.GetDamagePerHit(Manipulator.GetTransform(),Agent)
                TotalDamage:=TupleDamage(0)
                WeaponType:=TupleDamage(1)
                Critical:=TupleDamage(2)
                GameManager.ShowDamageToPlayer(TotalDamage,WeaponType,Agent,Critical)
                set CurrentHealth = CurrentHealth - TotalDamage
                Log("Current health : {GetOptimizedMagnitudeLetter(CurrentHealth)}, TotalDamage: {GetOptimizedMagnitudeLetter(TotalDamage)}", LOG_LEVEL_MINIGAMES)
                if(CurrentHealth<0.0):
                    set CurrentHealth = 0.0
                #Add agent to damaging agents
                if(var AllDamagingAgents : [agent]int = AllArcheryDamagingAgents[Index]):
                    if(set AllDamagingAgents[Agent] = 0, set AllArcheryDamagingAgents[Index] = AllDamagingAgents):
                #Update UI to all agents
                if(set AllArcheryMiniGame_CurrentHealth[Index] = CurrentHealth):
                    if(var AllDamagingAgents : [agent]int = AllArcheryDamagingAgents[Index]):
                        for(Player : GetPlayspace().GetPlayers(), OtherAgent:=agent[Player], DamageDelay:=AllDamagingAgents[OtherAgent]):
                            if(DamageDelay>5):
                                set AllDamagingAgents = RemoveAgentFromIntMap(AllDamagingAgents,OtherAgent)
                                if(AllArcheryDamagingAgents[Index]=AllDamagingAgents):
                            else:
                                GameManager.SetEntityHealth(OtherAgent, CurrentHealth, MaxHealth, ARROW_MINIGAME, Index+1)
                if(CurrentHealth=0.0):
                    TryRunArcheryPillars(Index)
            else:
                PrintFatalError("On PillarDamagedEvent: Index: {Index} invalid!")

    TryRunArcheryPillars(Index : int) : void =
        Log("Trying to move pillars {Index}", LOG_LEVEL_MINIGAMES)
        if(OldValue:=AllArcheryState[Index], set AllArcheryState[Index] = 0):
            if(OldValue=AllArcheryPillarsMaxTime[Index]):
                spawn:
                    RunArcheryPillars(Index)

    RunArcheryPillars(Index : int)<suspends> : void =
        #Move Pillar
        Log("Moving pillars Index :{Index}", LOG_LEVEL_MINIGAMES)
        if(GameManager:=MaybeGameManager?,PillarsSelected :=AllPillars[Index]):
            if(Pillar := PillarsSelected[0]):
                spawn:
                    GameManager.VFXManager.PlaySound(HIGHVALVE_SOUND,0.0,false,option{Pillar.GetTransform()})
            if(DamagingAgents := AllArcheryDamagingAgents[Index]):
                for(Player : GetPlayspace().GetPlayers(), PosibleAgent := agent[Player], Value := DamagingAgents[PosibleAgent]):
                    GameManager.VFXManager.SmallObjetive_AudioPlayer.Play(PosibleAgent)
            for(IndexPillar:=0..PillarsSelected.Length-1, Pillar := PillarsSelected[IndexPillar], OriginalPosition:=AllPillarsOriginalPositions[Index][IndexPillar]):
                Pillar.MoveTo(OriginalPosition,0.5)
            loop:
                Sleep(1.0)
                if(OldValue := AllArcheryState[Index], MaxTime := AllArcheryPillarsMaxTime[Index]):
                    NewValue := OldValue + 1
                    Log("ArcheryState: {NewValue}/{MaxTime}", LOG_LEVEL_MINIGAMES)
                    if(set AllArcheryState[Index]=NewValue):
                    if(NewValue=AllArcheryPillarsMaxTime[Index], MaxHealth := AllArcheryMiniGame_MaxHealth[Index], set AllArcheryMiniGame_CurrentHealth[Index] = MaxHealth):
                        if(Pillar := PillarsSelected[0]):
                            spawn:
                                GameManager.VFXManager.PlaySound(HIGHVALVE_SOUND,0.0,false,option{Pillar.GetTransform()})
                        for(IndexPillar:=0..PillarsSelected.Length-1, Pillar := PillarsSelected[IndexPillar], OriginalPosition:=AllPillarsOriginalPositions[Index][IndexPillar]):
                            ObjetiveTransform : transform =transform{Translation:=vector3{X:=OriginalPosition.Translation.X,Y:=OriginalPosition.Translation.Y,Z:=OriginalPosition.Translation.Z-ArcheryPillarsTranslation}, Rotation:=OriginalPosition.Rotation, Scale:=OriginalPosition.Scale}
                            spawn:
                                Pillar.MoveTo(ObjetiveTransform,0.5)
                            if(set AllArcheryDamagingAgents[Index] = map{}):
                            Log("Pillars returning to original state", LOG_LEVEL_MINIGAMES)
                        break
                else:
                    break

    TryRunArcheryRegeneration(Index : int) : void =
        if(IsRunningRegeneration:=AllArcheryMiniGame_RegenerationActive[Index]):
            if(State := AllArcheryState[Index], State=AllArcheryPillarsMaxTime[Index], not IsRunningRegeneration?, set AllArcheryMiniGame_RegenerationActive[Index]=true):
                spawn:
                    RunArcheryRegeneration(Index)
        else:
            PrintFatalError("Index {Index} invalid")

    RunArcheryRegeneration(Index : int)<suspends> : void =
        if(GameManager := MaybeGameManager?, Regeneration:=AllArcheryMiniGame_MaxHealth[Index]*ArcheryMiniGame_Regeneration):
            var NumSecondsWithoutDamage : int = 0
            var LastHealth : float = 0.0
            var AllAgentsDamaging : []agent = array{}
            loop:
                if(OldHealth := AllArcheryMiniGame_CurrentHealth[Index], OldHealth > 0.0):
                    #Archery
                    if(MaxHealth:=AllArcheryMiniGame_MaxHealth[Index], var NewHealth : float = OldHealth + Regeneration, set AllArcheryMiniGame_CurrentHealth[Index] = NewHealth):
                        if(LastHealth = OldHealth):
                            set NumSecondsWithoutDamage = NumSecondsWithoutDamage + 1
                        set LastHealth = NewHealth
                        if(NewHealth>=MaxHealth, set AllArcheryMiniGame_CurrentHealth[Index] = MaxHealth):
                            if(set AllArcheryMiniGame_RegenerationActive[Index] = false):
                                for(Player : GetPlayspace().GetPlayers(), Agent:=agent[Player], AllDamagingAgents:=AllArcheryDamagingAgents[Index], DamageDelay:=AllDamagingAgents[Agent]):
                                    if(5<DamageDelay):
                                        GameManager.SetEntityHealth(Agent, NewHealth, MaxHealth, ARROW_MINIGAME, Index+1)

                            break
                        else:
                            if(var AllDamagingAgents : [agent]int = AllArcheryDamagingAgents[Index]):
                                for(Player : GetPlayspace().GetPlayers(), Agent:=agent[Player], DamageDelay:=AllDamagingAgents[Agent]):
                                    if(DamageDelay>5):
                                        set AllDamagingAgents = RemoveAgentFromIntMap(AllDamagingAgents,Agent)
                                        if(set AllArcheryDamagingAgents[Index]=AllDamagingAgents):
                                    else:
                                        if(NumSecondsWithoutDamage<5):
                                            GameManager.SetEntityHealth(Agent, NewHealth, MaxHealth, ARROW_MINIGAME, Index+1)
                    #Degrade
                    if(var AllDamagingAgents : [agent]int = AllArcheryDamagingAgents[Index]):
                        for(Player : GetPlayspace().GetPlayers(), Agent:=agent[Player], Value:=AllDamagingAgents[Agent]):
                            if(set AllDamagingAgents[Agent] = Value + 1, set AllArcheryDamagingAgents[Index]=AllDamagingAgents):   
                else:
                    Log("Regeneration destroyed", LOG_LEVEL_MINIGAMES)
                    if(set AllArcheryMiniGame_RegenerationActive[Index] = false):
                    break    
                Sleep(1.0)

#INESTABLE BREAD
    AddInestableHB(InestableHB : hardbread_managerv2) : void =
        set AllInestableHB = AllInestableHB + array{InestableHB}

    HBInestabilityTick<private>() : void =
        for(Index:=0..AllInestableHB.Length-1,InestableHB := AllInestableHB[Index]):
            IsNeedsToBeMoved:=InestableHB.InestabilityTick(MaxHBInestableTime)
            if(IsNeedsToBeMoved?):
                NextInestablePosition:=FindAndAssingInestableHBPositon(InestableHB)
                spawn:
                    MoveInestableHB(InestableHB,NextInestablePosition)
                Log("Inestable bread {Index} moved to inestable position {NextInestablePosition}",LOG_LEVEL_MINIGAMES)

    FindAndAssingInestableHBPositon(InestableHB_ToMove : hardbread_managerv2) : int =
        var FinalInestablePosition : int = -1
        var IsValid: logic = true
        var NumLoops : int = 0
        loop:
            PosibleNewIndexPos:=GetRandomInt(0,AllInestablePosiblePositions.Length-1)
            if(InestablePropPosition:=AllInestablePosiblePositions[PosibleNewIndexPos]):
                LevelAreaPosition:=FindNearestHBLevel(InestablePropPosition.GetTransform())
                if(PosibleNewIndexPos<>InestableHB_ToMove.AssignedInestablePosition): #New posible position is diferent than the current position
                    for(HBIndex:=0..AllInestableHB.Length-1, Other_InestableHB := AllInestableHB[HBIndex], Other_InestableHB.AssignedInestablePosition<>InestableHB_ToMove.AssignedInestablePosition, Other_InestableHB.InestableHBIndex<>InestableHB_ToMove.InestableHBIndex  <#OtherHB is not this one#>): #Check if there is other InestableHB in this position
                        if(Other_InestableHB.AssignedInestablePosition=PosibleNewIndexPos):
                            Log("IntestableHB [{InestableHB_ToMove.InestableHBIndex}] tried to move to invalid position {PosibleNewIndexPos} because InestableHB [{HBIndex}] is using this position",LOG_LEVEL_DEBUG1)
                            set IsValid = false
                            #We need to check the level of the near Inestable HBs and check if there is other Inestable HB in the same level area
                        if(LevelAreaPosition=Other_InestableHB.Level):
                            Log("IntestableHB [{InestableHB_ToMove.InestableHBIndex}] tried to move to invalid position {PosibleNewIndexPos} because InestableHB [{HBIndex}] is in the same Level Area [{Other_InestableHB.Level}]",LOG_LEVEL_DEBUG1)
                            set IsValid = false
                    if(IsValid?):
                        Log("IntestableHB [{InestableHB_ToMove.InestableHBIndex}] found VALID POSITION at {PosibleNewIndexPos}",LOG_LEVEL_DEBUG1)
                        set FinalInestablePosition = PosibleNewIndexPos
                    break
            if(NumLoops>20):
                set FinalInestablePosition = InestableHB_ToMove.AssignedInestablePosition
                PrintWarning("Infinite loop detected on FindAndAssingInestableHBPositon")
                break
            set NumLoops = NumLoops + 1
        InestableHB_ToMove.SetInestablePosition(FinalInestablePosition)
        return FinalInestablePosition
                
    MoveInestableHB(InestableHardBread : hardbread_managerv2, NextInestablePosition : int)<suspends> : void =
        if(GameManager := MaybeGameManager?):
            if(InestablePosition:=AllInestablePosiblePositions[NextInestablePosition]):
                #Spawn Dirt
                DirtSpawnPosition:=AdaptTransformFromVisualEmpty(InestableHardBread.SlateBreadProp.GetTransform(),128.0)
                GameManager.GeneralMobsManager.MobBehaviourHelper.PlayDirtEffect(DirtSpawnPosition)
                #New position
                FinalTransform:=AdaptTransformFromVisualEmpty(InestablePosition.GetTransform(),VisualEmptyHideZ)
                if(InestableHardBread.SlateBreadProp.TeleportTo[FinalTransform]):
                    GameManager.GeneralMobsManager.MobBehaviourHelper.PlayDirtEffect(FinalTransform)
                else:
                    TeleportFailedAttempt(false, false, option{InestableHardBread.SlateBreadProp}, FinalTransform, 64.0)
                Sleep(1.0)
                NewLevel:=FindNearestHBLevel(InestableHardBread.SlateBreadProp.GetTransform())
                InestableHardBread.InitNewInestability(NewLevel)
        
        #Init Inestable Bread level
    FindNearestHBLevel(OriginTransform : transform) : int =
        var NearestHBTranslation: vector3 = vector3{X:=0.0,Y:=0.0,Z:=0.0}
        var NearestHB_Level : int = 0
        if(GameManager := MaybeGameManager?):
            for(StableHardBread : GameManager.AllHardBreads, not StableHardBread.IsInestable?):
                if(Distance(StableHardBread.SlateBreadProp.GetTransform().Translation,OriginTransform.Translation)<Distance(OriginTransform.Translation,NearestHBTranslation)):
                    set NearestHBTranslation = StableHardBread.SlateBreadProp.GetTransform().Translation
                    set NearestHB_Level = StableHardBread.Level
        return NearestHB_Level

    

    #STEVE
    SteveComputerCompletedEventCeremony(Index: int)<suspends> : void =
        Log("Executing Steve ceremony on Index: {Index}", LOG_LEVEL_DEBUG1)
        if(PropComputer := AllSteve_PropComputers[Index]):
            PropComputer.SetMaterial(Materials.GreenScreen_Inst1,?Index:=1)
            for(A:=0..5):
                PropComputer.SetMaterial(Materials.MI_Happy,?Index:=2)
                Sleep(0.5)
                PropComputer.SetMaterial(Materials.MI_Wink,?Index:=2)
                Sleep(0.5)
            Sleep(5.0)
            PropComputer.SetMaterial(Materials.RedScreen_Inst,?Index:=1)
            PropComputer.SetMaterial(Materials.Mi_Error,?Index:=2)
        
    
