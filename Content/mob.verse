# fill this file with your verse-script
using { /Verse.org/Simulation }
using { /Fortnite.com/Characters }
using { /Fortnite.com/Game }
using { /UnrealEngine.com/Temporary/SpatialMath }
using { /Fortnite.com/Devices }

mob := class():

    GeneralMobsManager : general_mobs_manager
    Assignation : int
    MaxLifeTime : float = RandomMobMaxLifeTime
    MaxTimeWithout_NearbyPlayers : float = 15.0
    var MaybePlayerOwner : ?player = false 
    var NumDungeon : int = -1
    var HBAreaLV : int = 1
    var IsBeingUsed : logic = false
    var IsBossChild : logic = false
    var MaybeNPCAgent : ?agent = false
    var MaybeNPCCharacter : ?fort_character = false
    var MobType : int = -1
    var LV : int = -1
    var LifeTime : float = 0.0
    var NoEnemiesTime : float = 0.0

    var MaxHealth : float = 0.0
    var CurrentHealth : float = 100.0
    var LastHitHealth : float = 100.0
    var DeathRunning : logic = false

    var MobBehaviourHelper : mob_behavior_helper = mob_behavior_helper{}

    var CurrentPosition : transform = transform{}

    var AllDamagingPlayers : [player]float = map{}

    Init() : void =
        if(MobType<>PLANT_MOB, NumDungeon = -1):
            PrintWarning("Ground mobs and druid mobs are not allowed to be play outside\n the dungeon because the Heal function can\n break the total award for elimination")
        spawn:
            SaveTransform()
        set MobBehaviourHelper = GeneralMobsManager.MobBehaviourHelper
        <#
        AllObjects:=FindCreativeObjectsWithTag(helper{})
        if(AllObjects.Length=1, Object := AllObjects[0], Helper := mob_behavior_helper[Object]):
            set MobBehaviourHelper = Helper
        else:
            PrintFatalError("FATAL ERROR on ground_mob_behaviour, CreativeObjectWithTag ``helper`` not found")#>
        if(NumDungeon=-1):
            spawn:
                RandomMobLifeLoop()
        set DeathRunning = false
        if(NPCCharacter := MaybeNPCCharacter?):
            if(GeneralMobsManager.HealthBarIsOrganic?):
                NPCCharacter.DamagedEvent().Subscribe(NPCOrganicDamaged)
                NPCCharacter.EliminatedEvent().Subscribe(NPCOrganicEliminatedEvent)
                NPCCharacter.HealedEvent().Subscribe(NPCOrganicHealedEvent)
                spawn:
                    DelaySetOrganicHealth()
            else:
                #NPCCharacter.DamagedEvent().Subscribe(NPCVirtualDamaged)
                #spawn:
                #    DelaySetVirtualHealth()

    SaveTransform()<suspends> : void =
        loop:
            if(NPCCharacter := MaybeNPCCharacter?, Transform := NPCCharacter.GetTransform()):
                Sleep(0.8)
                set CurrentPosition = Transform
                if(DeathRunning?):
                    break
            else:
                break

    SetPlayerOwner(MaybePlayer : ?player) : void =
        set MaybePlayerOwner = MaybePlayer

    DelaySetOrganicHealth()<suspends> : void =
        var ExpectedDamageReceived_PerHit : float = 100.0
        if(NumDungeon<>-1):
            if(Damage:=GeneralMobsManager.AllDungeons_ExpectedHammerDamage[NumDungeon]):
                set ExpectedDamageReceived_PerHit = Damage
                Log("MobSpawned of Dungeon {NumDungeon}. ExpectedDamageReceived_PerHit: {ExpectedDamageReceived_PerHit}", LOG_LEVEL_DEBUG9)
            else:
                PrintFatalError("On DelaySetOrganicHealth: NumDungeon: {NumDungeon} Invalid!")
        else:
            var DungeonEqualFound : int = 0
            var IndexDungeon : int = 0
            loop:
                if(HBAreaLevel_OfDungeon := GeneralMobsManager.AllDungeons_HBAreaLevel[IndexDungeon], HBAreaLevel_OfDungeon<=HBAreaLV):
                    set DungeonEqualFound = IndexDungeon
                else:
                    break
                set IndexDungeon = IndexDungeon + 1
            
            if(Damage:=GeneralMobsManager.AllDungeons_ExpectedHammerDamage[DungeonEqualFound]):
                set ExpectedDamageReceived_PerHit = Damage
                Log("Equal Mob.NumDungeon: {DungeonEqualFound}. ExpectedDamageReceived_PerHit: {Damage}", LOG_LEVEL_DEBUG9)
            else:
                PrintFatalError("On CalculateMobDamage: DungeonEqualFound: {DungeonEqualFound} Invalid!")
            

        Sleep(0.1)
        set IsBeingUsed = true
        
        set CurrentHealth=CalculeMobHealth(MobType,ExpectedDamageReceived_PerHit,HBAreaLV)  
        set MaxHealth = CurrentHealth
        set LastHitHealth = CurrentHealth
        if(NPCCharacter := MaybeNPCCharacter?):
            NPCCharacter.SetMaxHealth(MaxHealth)
            NPCCharacter.SetHealth(MaxHealth)
            ##Print("MaxHealth: {MaxHealth}")
        
    <#
    DelaySetVirtualHealth()<suspends> : void =
        Sleep(0.1)
        set IsBeingUsed = true
        if(NPCCharacter := MaybeNPCCharacter?, NPCCharacter.GetHealth()<>MOBSTARTINGHEALTH):
            PrintFatalError("NPCCharacter health if not MOBSTARTINGHEALTH {MOBSTARTINGHEALTH}")
        set CurrentHealth=CalculeMobHealth(MobType,LV,HBAreaLV)  
        set MaxHealth = CurrentHealth
        set LastHitHealth = CurrentHealth
        ##Print("MaxHealh: {MaxHealth}")#>

    NPCOrganicEliminatedEvent(EliminationResult : elimination_result) : void =
        <#
        ##Print("Eliminated event")
            RunDeath(option{ValidAgent})
        else:
            PrintWarning("Invalid eliminating character")#>
        if(ValidEliminatingCharacter  := EliminationResult.EliminatingCharacter?, EliminatingAgent := ValidEliminatingCharacter.GetAgent[]):
            if(not DeathRunning?):
                RunDeath(false)
            if(MobType=MINIBOSS_MOB, AllAgentsPlaying := GeneralMobsManager.AllDungeons_AgentsPlaying[NumDungeon]):
                for(AgentPlaying : AllAgentsPlaying, GameManager := GeneralMobsManager.MaybeGameManager?, AgentBakery := GameManager.GetAgentBakery_Transacted(AgentPlaying)?, HealthUI := AgentBakery.MaybeEntityHealthUI?):
                    HealthUI.SetInfo(0.0,MaxHealth,MobType,NumDungeon+1)    
        else:
            PrintFatalError("On NPCOrganicEliminatedEvent, Invalid EliminatingCharacter!")
            
    <#
    DealDamage(DamageToDeal : float) : void =
        if:
            NPCCharacter := MaybeNPCCharacter?
        then:
            NPCCharacter.Damage(DamageToDeal)#>

    NPCOrganicDamaged(DamageResult : damage_result) : void =
        MaybeInstigator:=DamageResult.Instigator
        MaybeSource:=DamageResult.Source
        OrganicDamageAmount:=DamageResult.Amount
        Character:=DamageResult.Target
        if:
            NPCCharacter := MaybeNPCCharacter?
            NPCCharacter.GetHealth()>0.0
        then: 
            if(ValidInstigator := MaybeInstigator?, AgentInstigator := ValidInstigator.GetInstigatorAgent[]):
                #Add real damage
                ##Print("PreHealthDamage: {NPCCharacter.GetHealth()}")
                TupleDamage:=GeneralMobsManager.GetDamagePerHit(NPCCharacter.GetTransform(),AgentInstigator)
                AmountTotalToDamage := TupleDamage(0)
                if(AmountTotalToDamage>OrganicDamageAmount):
                    NPCCharacter.Damage(AmountTotalToDamage-OrganicDamageAmount)
                    
                ShowDamageToPlayer(AmountTotalToDamage,TupleDamage(1),AgentInstigator,TupleDamage(2))
                set LastHitHealth = CurrentHealth

                if(DamagingPlayer:=player[AgentInstigator]):
                    var AmountToSave : float = AmountTotalToDamage
                    var OldAmount : float = 0.0
                    if(Value:=AllDamagingPlayers[DamagingPlayer], set OldAmount = Value):
                    if(AmountToSave>=LastHitHealth):
                        set AmountToSave = LastHitHealth
                    ##Print("Final amount to save: {AmountToSave} accumulated: {AmountToSave+OldAmount}")
                    if(set AllDamagingPlayers[DamagingPlayer] = AmountToSave+OldAmount):
            else:
                PrintFatalError("Invalid instigator")  
                ##Print("Organic damage: {TupleDamage(0)}, CurrentHealth: {NPCCharacter.GetHealth()}")
            if(MobType=MINIBOSS_MOB, AllAgentsPlaying := GeneralMobsManager.AllDungeons_AgentsPlaying[NumDungeon]):
                for(AgentPlaying : AllAgentsPlaying, GameManager := GeneralMobsManager.MaybeGameManager?, AgentBakery := GameManager.GetAgentBakery_Transacted(AgentPlaying)?, HealthUI := AgentBakery.MaybeEntityHealthUI?):
                    HealthUI.SetInfo(NPCCharacter.GetHealth(),MaxHealth,ENTITY_BOSS,NumDungeon+1)    

    NPCOrganicHealedEvent(HealingResult : healing_result) : void =
        AmountHealed:=HealingResult.Amount
        if(NPCCharacter := MaybeNPCCharacter?):
            if(NPCCharacter.GetHealth()>=NPCCharacter.GetMaxHealth()):
                set AllDamagingPlayers = map{} #Clean map if it is full health
            else:
                PercentageNotHealed : float = 1.0-(AmountHealed/NPCCharacter.GetMaxHealth())
                for(Player : GeneralMobsManager.GetPlayspace().GetPlayers(), Amount := AllDamagingPlayers[Player], set AllDamagingPlayers[Player] = Amount*PercentageNotHealed):
                   


    ShowDamageToPlayer(Damage : float, WeaponType : int, Agent : agent, IsCritical : logic): void =
        ##Print("Trying to show damage to player on UI")
        if:
            GameManager:=GeneralMobsManager.MaybeGameManager?
        then:
            if(WeaponType=BOW_TYPE):
                GameManager.DisplayResourcersGather(Agent, array{Damage}, array{BOW_TYPE}, IsCritical)
            else if(WeaponType=SWORD_TYPE):
                GameManager.DisplayResourcersGather(Agent, array{Damage}, array{SWORD_TYPE},IsCritical)
            else if(WeaponType=-1):
                GameManager.DisplayResourcersGather(Agent, array{Damage}, array{SWORD_TYPE},IsCritical)
        else:
            PrintFatalError("On ShowDamageToPlayer, GameManager not found")

            
            <#
    NPCVirtualDamaged(DamageResult : damage_result) : void =
        MaybeInstigator:=DamageResult.Instigator
        MaybeSource:=DamageResult.Source
        Character:=DamageResult.Target
        if:
            NPCCharacter := MaybeNPCCharacter?
            ValidInstigator := MaybeInstigator?
            NPCCharacter.GetHealth()>0.0
            AgentInstigator := ValidInstigator.GetInstigatorAgent[]
        then: 
            #Add real damage
            TupleDamage:=GeneralMobsManager.GetDamagePerHit(NPCCharacter.GetTransform(),AgentInstigator)
            var ColateralDamage : float = 0.0 
            ##Print("Virtual damaged. Last hit health: {LastHitHealth}")
            if(LastHitHealth=CurrentHealth):
                #Last hit health is the same, so the damage of the sword was not applied
                set ColateralDamage = 0.0
            else if(LastHitHealth>CurrentHealth):
                #Last hit health is bigger same, so the damage of the sword was applied
                set ColateralDamage = DamageResult.Amount

            ##Print("Damage: {TupleDamage(0)}, CurrentHealth: {CurrentHealth}, ColateralDamage: {ColateralDamage}. Set to: {CurrentHealth - TupleDamage(0) + ColateralDamage}")
            ShowDamageToPlayer(TupleDamage(0),TupleDamage(1),AgentInstigator,TupleDamage(2))
            set CurrentHealth = CurrentHealth - TupleDamage(0) + ColateralDamage
            set LastHitHealth = CurrentHealth
            Percentage := CurrentHealth/MaxHealth
            VirtualAmmount := Percentage*MOBSTARTINGHEALTH
            if(VirtualAmmount<=0.0, not DeathRunning?):
                NPCCharacter.SetHealth(1.0)
                RunDeath()
            else:
                ##Print("Tring to set the NPC health to {Percentage*MOBSTARTINGHEALTH}")
                NPCCharacter.SetHealth(Percentage*MOBSTARTINGHEALTH)#>

    DelayGrantAward<private>()<suspends> : void =
        Sleep(0.25)
        if(NumDungeon=-1):
            GeneralMobsManager.GrantMobDestroyedReward(MaybePlayerOwner,MaxHealth,MaxHealth,MobType,LV,NumDungeon,HBAreaLV)
            #Grant award to all player that have damaged this mob
            #for(Player : GeneralMobsManager.GetPlayspace().GetPlayers(), Amount := AllDamagingPlayers[Player]):
            #    GeneralMobsManager.GrantMobDestroyedReward(option{Player},Amount,MaxHealth,MobType,LV,NumDungeon,HBAreaLV)
        else:
            GeneralMobsManager.GrantMobDestroyedReward(false,0.0,MaxHealth,MobType,LV,NumDungeon,HBAreaLV)
        set AllDamagingPlayers = map{}


    DespawnMobV2(IsForced : logic) : void =
        <#
        if(not RewardAllowed?):
            set AllDamagingPlayers = map{}
            set NumDungeon = -1 #NumDungeon -1 because the mobs that are not from a dungeon does not grant reward
            #>
        RunDeath(IsForced)
        

    RunDeath(IsForced : logic) : void =
        set DeathRunning = true
        if(IsForced?):
            if:
                NPCCharacter := MaybeNPCCharacter?
            then:
                NPCCharacter.Damage(NPCCharacter.GetMaxHealth()+10.0)
        #Death effect
        MobBehaviourHelper.PlayDeathEffect(MobType,CurrentPosition)
        if(not IsForced?):
            spawn:
                DelayGrantAward()
        if(NumDungeon <> -1):
            if(MobType=MINIBOSS_MOB):
                GeneralMobsManager.OnMiniBossDeath(NumDungeon)
        if(not GeneralMobsManager.HealthBarIsOrganic?):
            DamageArgs : damage_args = damage_args{Amount:=2.0}
            if(NPCCharacter := MaybeNPCCharacter?):
                NPCCharacter.Damage(DamageArgs)
        spawn:
            DelayCleanMemory()
            

    DelayCleanMemory()<suspends> : void =
        set DeathRunning = true
        Sleep(1.0)
        set LifeTime = 0.0
        set NoEnemiesTime = 0.0
        set MaybeNPCAgent = false
        set MaybeNPCCharacter = false
        set AllDamagingPlayers = map{}
        GeneralMobsManager.CleanNPCMemory(Assignation)

    SetAgentAndCharacter(InNPCAgent : agent, InNPCCharacter : fort_character) : void =
        set MaybeNPCAgent = option{InNPCAgent}
        set MaybeNPCCharacter = option{InNPCCharacter}

    SetStats(InMobType : int, InLV : int, InNumDungeon : int, InIsBossChild : logic, InHBAreaLV : int) : void =
        set IsBossChild = InIsBossChild
        set MobType = InMobType
        set LV = InLV
        set HBAreaLV = InHBAreaLV
        set NumDungeon = InNumDungeon

    SetUse(InIsBeingUsed : logic) : void =
        set IsBeingUsed = InIsBeingUsed

    #DamagePlayer()
    RandomMobLifeLoop()<suspends>: void =
        if(NumDungeon=-1):
            #LifeLoopAllowed
            TickTime := 0.5
            loop:
                Sleep(TickTime)
                set LifeTime = LifeTime + TickTime
                IsThereEnemies:=IsThereNearbyPlayers()
                if(not IsThereEnemies?):
                    set NoEnemiesTime = NoEnemiesTime + TickTime
                if(MaxLifeTime<LifeTime or MaxTimeWithout_NearbyPlayers<NoEnemiesTime):
                    DespawnMobV2(true)
                    break
                if(DeathRunning?):
                    break  

    IsThereNearbyPlayers() : logic =
        var IsThereEnemies : logic = false
        if(NPCCharacter:=MaybeNPCCharacter?):
            for(Player : GeneralMobsManager.GetPlayspace().GetPlayers(), PlayerCharacter:=Player.GetFortCharacter[],1024.0>Distance(PlayerCharacter.GetTransform().Translation,NPCCharacter.GetTransform().Translation)):
                set IsThereEnemies = true
        return IsThereEnemies

MakeMobManagerUI<constructor>(InGeneralMobsManager : general_mobs_manager, InAssignation : int)<transacts> := mob :
        GeneralMobsManager := InGeneralMobsManager
        Assignation := InAssignation
