# See https://dev.epicgames.com/documentation/en-us/uefn/create-your-own-device-in-verse for how to create a verse device.

using { /Fortnite.com/Devices }
using { /UnrealEngine.com/Temporary/SpatialMath }
using { /Fortnite.com/Characters }
using { /Verse.org/Simulation }
using { /Verse.org/Random }
using { /Fortnite.com/AI }
using { /Fortnite.com/Animation/PlayAnimation }
using { /Verse.org/Assets }
using { /UnrealEngine.com/Temporary/Diagnostics }
using { /Verse.org/Colors }

mob_behavior_debug_draw := class(debug_draw_channel) {}

# A Verse-authored creative device that can be placed in a level
mob_behavior_helper := class(creative_device):

    var MaybeGeneralMobsManager : ?general_mobs_manager = false

    @editable
    MobDeathEffect : creative_prop_asset = DefaultCreativePropAsset
    @editable
    MiniBossDeathEffect : creative_prop_asset = DefaultCreativePropAsset


    #Bitesounds
    @editable
    AllAudioPlayer_LoadBite0Sound : []audio_player_device = array{}
    var AllAudioPlayerLoadBite0_Used : []logic = array{}
    @editable
    AllAudioPlayer_Bite0Sound : []audio_player_device = array{}
    var AllAudioPlayerBite0_Used : []logic = array{}

    @editable
    AllAudioPlayer_LoadBite1Sound : []audio_player_device = array{}
    var AllAudioPlayerLoadBite1_Used : []logic = array{}
    @editable
    AllAudioPlayer_Bite1Sound : []audio_player_device = array{}
    var AllAudioPlayerBite1_Used : []logic = array{}

    #Others
    @editable
    AllAudioPlayer_BossLaugh : []audio_player_device = array{}
    var AllAudioPlayerBossLaugh_Used : []logic = array{}

    @editable
    AllAudioPlayer_BossBiteSound : []audio_player_device = array{}
    var AllAudioPlayerBossBite_Used : []logic = array{}
    @editable
    AllAudioPlayer_LoadBossBiteSound : []audio_player_device = array{}
    var AllAudioPlayerLoadBossBite_Used : []logic = array{}

    @editable
    AllAudioPlayer_DeathSound : []audio_player_device = array{}
    var AllAudioPlayerDeath_Used : []logic = array{}

    @editable
    AllAudioPlayer_DirtSound : []audio_player_device = array{}
    var AllAudioPlayerDirt_Used : []logic = array{}

    @editable
    AllAudioPlayer_LoadZone : []audio_player_device = array{}
    var AllAudioLoadZone_Used : []logic = array{}

    @editable
    AllAudioPlayer_DamageZone : []audio_player_device = array{}
    var AllAudioDamageZone_Used : []logic = array{}

    @editable
    AllAudioPlayer_MobHealerZone : []audio_player_device = array{}
    var AllAudioMobHealerZone_Used : []logic = array{}

    #@editable
    #BP_Projectile : creative_prop_asset = DefaultCreativePropAsset

    @editable
    BP_SpitDamage_Short : creative_prop_asset = DefaultCreativePropAsset
    @editable
    BP_SpitDamage_Long : creative_prop_asset = DefaultCreativePropAsset

    @editable
    BP_DamageZone : creative_prop_asset = DefaultCreativePropAsset

    @editable
    BP_MiniBossShieldZone : creative_prop_asset = DefaultCreativePropAsset

    @editable
    BP_MobHealZone : creative_prop_asset = DefaultCreativePropAsset

    ProjectileSpeed_CmPerSecond : float = 10000.0

    @editable
    MobsCanMove : logic = true

    var TotalTime : int = 0

    SmallSpitRange : float = 242.0
    LongSpitRange : float = 362.0
    MiniBossSpitRange : float = 362.0

    SmallSpitDuration : float = 1.0
    LongSpitDuration : float = 1.0
    MiniBossSpitDuration : float = 0.5

    ShortAttack_PreDelay : float = 1.125
    ShortAttack_PostDelay : float = 0.25

    LongAttack_PreDelay : float = 1.224
    LongAttack_PostDelay : float = 0.375

    MiniBossAttack_PreDelay : float = 1.26
    MiniBossAttack_PostDelay : float = 0.2

    #MOBS
    Mob_EnemyRangeDetection : float = 1500.0
    Mob_NavigationErrorWait : float = 2.0

    #DamageZone
    DamageZonePreDamage_Delay : float = 2.0
    DamageZonePostDamage_Delay : float = 3.5
    DamageZoneRadius : float = 335.0

    #MiniBoss
    MobNoTarget_MoveToWait_Duration : float = 0.1

    MiniBoss_MaxSpecialActions : int = 3
    MiniBoss_WaitBetweenSpecialActions : int = 15

    MiniBoss_Max_Movement  : float = 512.0
    MiniBoss_CheckArena : float = 1500.0
    MiniBossBite_Damage : float = 20.0
    MiniBoss_AreaDamage : float = 10.0
    MiniBoss_MoveToWait_Duration : float = 0.25
    MiniBoss_ActionToWait_Duration : float = 0.5
    MiniBoss_MinionsSpawnToWait_Duration : float = 2.0

    MiniBossAttack_Range : float = 256.0

    <#
    #Flying mob
    FlyingMob_MoveToWait_Duration : float = 0.15
    FlyingMob_Damage : float = 5.0
    FlyingMobAttack_Range : float = 128.0
    FlyingMobAttack_PreDelay : float = 0.5
    FlyingMobAttack_PostDelay : float = 1.0#>

    #GroundMobBite_Damage : float = 5.0

    #Ground mob
    GroundMob_Max_Movement : float = 1024.0
    GroundMob_MoveToWait_Duration : float = 0.2
    GroundMob_Damage : float = 15.0
    GroundMobAttack_Range : float = 188.0
    #GroundMobAttack_PreDelay : float = 1.6
    #GroundMobAttack_PostDelay : float = 0.5

    #Druid mob
    Druid_WaitBetweenSpecialActions : int = 10

    #HealZone
    HealZonePreHeal_Delay : float = 2.0
    HealZonePostHeal_Delay : float = 3.5
    MobHealZoneRadius : float = 335.0

    DruidMob_Max_Movement : float = 1024.0
    DruidMob_MoveToWait_Duration : float = 0.2
    DruidMob_Damage : float = 10.0
    DruidMobAttack_Range : float = 188.0
    DruidMobAttack_PreDelay : float = 1.6
    DruidMobAttack_PostDelay : float = 0.7

    #Plant Mob
    
    PreSpitWait : float = 0.5
    PostSpitWait : float = 0.5

    PlantMob_MoveToWait_Duration : float = 0.1
    PlantMob_Damage : float = 25.0
    PlantMobAttack_Range : float = 318.0

    #Animation
    #var FlyingMobAttack  : animation_sequence = EnemyMobs.MMB_Flying_01_Anim_Attack
    var DruidMobAttack  : animation_sequence = EnemyMobs.MutantMiniMob_Druid_Anim_AttackBite
    var GroundMobAttack  : animation_sequence = EnemyMobs.MutantMiniMob_Anim_AttackBite
    var MiniBossMobAttack  : animation_sequence = EnemyMobs.B_1_Anim_Attack
    var MiniBossMobLaugh  : animation_sequence = EnemyMobs.B_1_Anim_Risa
    var PlantMobAttack  : animation_sequence = EnemyMobs.Plant_Anim_Attack_001

    DebugDraw:debug_draw = debug_draw{Channel := mob_behavior_debug_draw}


    MakeActionV2(NPCTransform : transform, NewTargetCharacter : fort_character, MobType : int, NPCSpawnPoint : transform, LastSpecialAction : int, NumSpecialActions : int, SecondaryDamage : float) : tuple(logic,logic,?creative_prop,float,float,?animation_sequence,logic) =
        var IsSpecialActionExecuted : logic = false
        var IsZoneAttack : logic = false
        var MustAttack : logic = false
        var PreActionDelay : float = 0.2
        var PostActionDelay : float = 0.2
        var MaybeAreaCreative_Prop : ?creative_prop = false
        var MaybeAnimation : ?animation_sequence = false
        if(NewTargetCharacter.IsActive[], TargetAgent := NewTargetCharacter.GetAgent[]):
            if(MobType=MINIBOSS_MOB, GeneralMobsManager := MaybeGeneralMobsManager?): 
                NumDungeon:=GeneralMobsManager.FindNearDungeon(NPCTransform.Translation)
                GeneralMobsManager.CheckBossMusic(NumDungeon)
                if(NumSpecialActions<MiniBoss_MaxSpecialActions, TotalTime>LastSpecialAction+MiniBoss_WaitBetweenSpecialActions):
                    #Checking num of mobs spawned in this dungeon
                    var Num_Child_Spawned : int = 0
                    var Num_NoChild_Spawned : int = 0
                    var AllNoChildMobsSpawned : []mob = array{}
                    var AllChildMobsSpawned : []mob  = array{}
                    var IsSpawnMinionsAllowed : logic = false
                    for(Mob : GeneralMobsManager.AllMobsManager, Mob.MobType<>MINIBOSS_MOB, not Mob.DeathRunning?, Mob.IsBeingUsed?, Mob.NumDungeon=NumDungeon):
                        if(Mob.IsBossChild?):
                            set Num_Child_Spawned = Num_Child_Spawned + 1
                            set AllChildMobsSpawned = AllChildMobsSpawned + array{Mob}
                        else:
                            set Num_NoChild_Spawned = Num_NoChild_Spawned + 1
                            set AllNoChildMobsSpawned = AllNoChildMobsSpawned + array{Mob}
                    #Checking number of mobs in this dungeon
                    #Print("NumMobsSpawned: {Num_Child_Spawned}")
                    #Print("Num No Child: {Num_NoChild_Spawned}")
                    if(Num_NoChild_Spawned>10):
                        for(Mob : AllNoChildMobsSpawned):
                            #Print("Destroying  no child minions there are too many")
                            Mob.DespawnMobV2(true)
                        set IsSpawnMinionsAllowed = true
                    else if(Num_NoChild_Spawned > 4, Num_Child_Spawned > 4):
                        for(Mob : AllNoChildMobsSpawned):
                            #Print("Destroying  no child minions because there is not space")
                            Mob.DespawnMobV2(true)
                        set IsSpawnMinionsAllowed = true
                    else if(Num_Child_Spawned > 9):
                        for(Mob : AllNoChildMobsSpawned):
                            #Print("Destroying  no child minions for space")
                            Mob.DespawnMobV2(true)
                        set IsSpawnMinionsAllowed = true
                    else:
                        set IsSpawnMinionsAllowed = true


                    if(IsSpawnMinionsAllowed?):
                        set MaybeAnimation = option{MiniBossMobLaugh}
                        spawn:
                            PlaySound(NPCTransform,SOUND_BOSS_LAUGH,1.0)
                        set IsSpecialActionExecuted = true
                        #Print("Special action, spawn minions")
                        spawn:
                            SpawnMiniBossMinions(NPCTransform,NPCSpawnPoint.Translation)
                        set PostActionDelay = MiniBoss_MinionsSpawnToWait_Duration
                else:
                    if(300.0>DistanceXY(NewTargetCharacter.GetTransform().Translation,NPCTransform.Translation)):
                        if(0=GetRandomInt(0,1)):
                            set PreActionDelay=MiniBossAttack_PreDelay
                            set PostActionDelay=MiniBossAttack_PostDelay
                            set MustAttack = true
                            set MaybeAnimation = option{MiniBossMobAttack}
                        else:
                            set PreActionDelay=DamageZonePreDamage_Delay
                            set PostActionDelay=DamageZonePostDamage_Delay
                            set IsZoneAttack = true
                            spawn:
                                CreateDamageZone(NPCTransform,SecondaryDamage)
                    else:
                        set PreActionDelay=DamageZonePreDamage_Delay
                        set PostActionDelay=DamageZonePostDamage_Delay
                        set IsZoneAttack = true
                        spawn:
                            CreateDamageZone(NPCTransform,SecondaryDamage)
            else if(MobType=DRUID_MOB):
                if(TotalTime>LastSpecialAction+Druid_WaitBetweenSpecialActions):
                    set IsSpecialActionExecuted = true
                    set PreActionDelay = HealZonePreHeal_Delay
                    set PostActionDelay = HealZonePostHeal_Delay
                    HealingZoneTransform:=NPCTransform
                    spawn:
                        CreateHealZone(NPCTransform)
                else:
                    if(DruidMobAttack_Range>DistanceXY(NewTargetCharacter.GetTransform().Translation,NPCTransform.Translation)):
                        set PreActionDelay=DruidMobAttack_PreDelay
                        set PostActionDelay=DruidMobAttack_PostDelay
                        set MustAttack = true
                        set MaybeAnimation = option{DruidMobAttack}
            else if(MobType=GROUND_MOB):
                if(GroundMobAttack_Range>DistanceXY(NewTargetCharacter.GetTransform().Translation,NPCTransform.Translation)):
                    set PreActionDelay=ShortAttack_PreDelay
                    set PostActionDelay=ShortAttack_PostDelay
                    set MustAttack = true
                    set MaybeAnimation = option{GroundMobAttack}

            else if(MobType=PLANT_MOB):
                if(PlantMobAttack_Range>DistanceXY(NewTargetCharacter.GetTransform().Translation,NPCTransform.Translation)):
                    set PreActionDelay=LongAttack_PreDelay
                    set PostActionDelay=LongAttack_PostDelay
                    set MustAttack = true
                    set MaybeAnimation = option{PlantMobAttack}
        else:
            #There is no player on area...waiting
            set PreActionDelay = MobNoTarget_MoveToWait_Duration
            set PostActionDelay = MobNoTarget_MoveToWait_Duration
            #Sleep(PlantMob_MoveToWait_Duration)
            #Sleep(0.1)
        return (IsSpecialActionExecuted,MustAttack,MaybeAreaCreative_Prop,PreActionDelay,PostActionDelay,MaybeAnimation,IsZoneAttack)

<#
    MakeAction(NPCTransform : transform, NewTargetCharacter : fort_character, MobType : int, NPCSpawnPoint : transform, LastSpecialAction : int, NumSpecialActions : int)<suspends> : tuple(logic,logic,?creative_prop,float,float,?animation_sequence) =
        var IsSpecialActionExecuted : logic = false
        var MustAttack : logic = false
        var PreActionDelay : float = 1.0
        var PostActionDelay : float = 1.0
        var MaybeAreaCreative_Prop : ?creative_prop = false
        var MaybeAnimation : ?animation_sequence = false
        if(NewTargetCharacter.IsActive[], TargetAgent := NewTargetCharacter.GetAgent[]):
            if(MobType=MINIBOSS_MOB, GeneralMobsManager := MaybeGeneralMobsManager?): 
                NumDungeon:=GeneralMobsManager.FindNearDungeon(NPCTransform.Translation)
                GeneralMobsManager.CheckBossMusic(NumDungeon)
                if(NumSpecialActions<MiniBoss_MaxSpecialActions, TotalTime>LastSpecialAction+MiniBoss_WaitBetweenSpecialActions):
                    #Checking num of mobs spawned in this dungeon
                    var Num_Child_Spawned : int = 0
                    var Num_NoChild_Spawned : int = 0
                    var AllNoChildMobsSpawned : []mob = array{}
                    var AllChildMobsSpawned : []mob  = array{}
                    var IsSpawnMinionsAllowed : logic = false
                    for(Mob : GeneralMobsManager.AllMobsManager, Mob.MobType<>MINIBOSS_MOB, not Mob.DeathRunning?, Mob.IsBeingUsed?, Mob.NumDungeon=NumDungeon):
                        if(Mob.IsBossChild?):
                            set Num_Child_Spawned = Num_Child_Spawned + 1
                            set AllChildMobsSpawned = AllChildMobsSpawned + array{Mob}
                        else:
                            set Num_NoChild_Spawned = Num_NoChild_Spawned + 1
                            set AllNoChildMobsSpawned = AllNoChildMobsSpawned + array{Mob}
                    #Checking number of mobs in this dungeon
                    #Print("NumMobsSpawned: {Num_Child_Spawned}")
                    #Print("Num No Child: {Num_NoChild_Spawned}")
                    if(Num_NoChild_Spawned>10):
                        for(Mob : AllNoChildMobsSpawned):
                            #Print("Destroying  no child minions there are too many")
                            Mob.DealDamage(999999999999999999999.9)
                        set IsSpawnMinionsAllowed = true
                    else if(Num_NoChild_Spawned > 4, Num_Child_Spawned > 4):
                        for(Mob : AllNoChildMobsSpawned):
                            #Print("Destroying  no child minions because there is not space")
                            Mob.DealDamage(999999999999999999999.9)
                        set IsSpawnMinionsAllowed = true
                    else if(Num_Child_Spawned > 9):
                        for(Mob : AllNoChildMobsSpawned):
                            #Print("Destroying  no child minions for space")
                            Mob.DealDamage(999999999999999999999.9)
                        set IsSpawnMinionsAllowed = true
                    else:
                        set IsSpawnMinionsAllowed = true


                    if(IsSpawnMinionsAllowed?):
                        PlaySound(NPCTransform,SOUND_BOSS_LAUGH,1.0)
                        set IsSpecialActionExecuted = true
                        #Print("Special action, spawn minions")
                        spawn:
                            SpawnMiniBossMinions(NPCTransform,NPCSpawnPoint.Translation)
                        Sleep(MiniBoss_MinionsSpawnToWait_Duration)
                else:
                    if(300.0>DistanceXY(NewTargetCharacter.GetTransform().Translation,NPCTransform.Translation)):
                        if(0=GetRandomInt(0,1)):
                            set PreActionDelay=MiniBossAttack_PreDelay
                            set PostActionDelay=MiniBossAttack_PostDelay
                            set MustAttack = true
                            set MaybeAnimation = option{MiniBossMobAttack}
                        else:
                            CreateDamageZone(NPCTransform)
                    else:
                        CreateDamageZone(NPCTransform)
            else if(MobType=DRUID_MOB):
                if(TotalTime>LastSpecialAction+Druid_WaitBetweenSpecialActions):
                    set IsSpecialActionExecuted = true
                    #Print("Special action, spawn healing all mob allies in the area")
                    HealingZoneTransform:=NPCTransform
                    spawn:
                        PlaySound(NPCTransform,SOUND_LOAD_ZONE,DamageZonePreDamage_Delay)
                    set MaybeAreaCreative_Prop = SpawnMobHealZone(NPCTransform)
                    Sleep(HealZonePreDamage_Delay)
                    spawn:
                        PlaySound(NPCTransform,SOUND_HEAL_ZONE,-1.0)
                    if(GeneralMobsManager := MaybeGeneralMobsManager?):
                        for(Mob : GeneralMobsManager.AllMobsManager, not Mob.DeathRunning?, Mob.IsBeingUsed?):
                            if(OtherNPCCharacter := Mob.MaybeNPCCharacter?, Transform := OtherNPCCharacter.GetTransform()):
                                if(MobHealZoneRadius>DistanceXY(OtherNPCCharacter.GetTransform().Translation,HealingZoneTransform.Translation)):
                                    OtherNPCCharacter.Heal(OtherNPCCharacter.GetMaxHealth()*0.33)
                                    #Print("Character on zone detected, healing him")
                        Sleep(HealZonePostDamage_Delay)
                else:
                    if(DruidMobAttack_Range>DistanceXY(NewTargetCharacter.GetTransform().Translation,NPCTransform.Translation)):
                        set PreActionDelay=DruidMobAttack_PreDelay
                        set PostActionDelay=DruidMobAttack_PostDelay
                        set MustAttack = true
                        set MaybeAnimation = option{DruidMobAttack}
            else if(MobType=GROUND_MOB):
                if(GroundMobAttack_Range>DistanceXY(NewTargetCharacter.GetTransform().Translation,NPCTransform.Translation)):
                    set PreActionDelay=ShortAttack_PreDelay
                    set PostActionDelay=ShortAttack_PostDelay
                    set MustAttack = true
                    set MaybeAnimation = option{GroundMobAttack}

            else if(MobType=PLANT_MOB):
                if(PlantMobAttack_Range>DistanceXY(NewTargetCharacter.GetTransform().Translation,NPCTransform.Translation)):
                    set PreActionDelay=LongAttack_PreDelay
                    set PostActionDelay=LongAttack_PostDelay
                    set MustAttack = true
                    set MaybeAnimation = option{PlantMobAttack}
        else:
            #There is no player on area...waiting
            Sleep(PlantMob_MoveToWait_Duration)
            Sleep(0.1)
        return (IsSpecialActionExecuted,MustAttack,MaybeAreaCreative_Prop,PreActionDelay,PostActionDelay,MaybeAnimation)#>

    CreateHealZone(NPCTransform : transform)<suspends> : void =      
        spawn:
            PlaySound(NPCTransform,SOUND_LOAD_ZONE,DamageZonePreDamage_Delay)
        spawn:
            SpawnMobHealZone(NPCTransform)
        Sleep(HealZonePreHeal_Delay)
        spawn:
            PlaySound(NPCTransform,SOUND_HEAL_ZONE,-1.0)
        if(GeneralMobsManager := MaybeGeneralMobsManager?):
            for(Mob : GeneralMobsManager.AllMobsManager, not Mob.DeathRunning?, Mob.IsBeingUsed?):
                if(OtherNPCCharacter := Mob.MaybeNPCCharacter?, Transform := OtherNPCCharacter.GetTransform()):
                    if(MobHealZoneRadius>DistanceXY(OtherNPCCharacter.GetTransform().Translation,NPCTransform.Translation)):
                        OtherNPCCharacter.Heal(OtherNPCCharacter.GetMaxHealth()*0.25)
                        ##Print("Character on zone detected, healing him")
    
    
    CreateDamageZone<private>(NPCTransform : transform, BaseDamage : float)<suspends> : void =
        ##Print("Creating attack zone")
        Log("Creating DamageZone with damage {GetOptimizedMagnitudeLetter(BaseDamage)}",LOG_LEVEL_DEBUG7)
        if(GeneralMobsManager:=MaybeGeneralMobsManager?):
            NumDungeon:=GeneralMobsManager.FindNearDungeon(NPCTransform.Translation)
            DamageZoneTransform:transform=transform{Translation:=NPCTransform.Translation+vector3{Z:=-84.0},Rotation:=NPCTransform.Rotation,Scale:=NPCTransform.Scale}
            TupleSpawn:=SpawnProp(BP_DamageZone,DamageZoneTransform)
            MaybeCreativeProp:=TupleSpawn(0)
            if(ValidProp := MaybeCreativeProp?):
                spawn:
                    PlaySound(DamageZoneTransform,SOUND_LOAD_ZONE,DamageZonePreDamage_Delay)
                Sleep(DamageZonePreDamage_Delay)
                spawn:
                    PlaySound(DamageZoneTransform,SOUND_DAMAGE_ZONE,-1.0)
                for(Player : GetPlayspace().GetPlayers(), PlayerCharacter := Player.GetFortCharacter[]):
                    if(DamageZoneRadius>DistanceXY(PlayerCharacter.GetTransform().Translation,DamageZoneTransform.Translation)):
                        ##Print("Player on damage zone")
                        if(PlayerCharacter.IsActive[]):
                            DamageCharacter(PlayerCharacter,MiniBoss_AreaDamage)
                Sleep(DamageZonePostDamage_Delay)
                ValidProp.Dispose()

    DamageCharacter(CharacterObjetive : fort_character, DamageAmount : float) : void =
        if(CharacterObjetive.IsActive[]):
            CharacterObjetive.Damage(DamageAmount)
            if(Agent := CharacterObjetive.GetAgent[], GeneralMobsManager:=MaybeGeneralMobsManager?, GameManager := GeneralMobsManager.MaybeGameManager?):
                GameManager.VFXManager.Damage_AudioPlayer.Play(Agent)
                MaybeBakery:=GameManager.GetAgentBakery(Agent)
                if(ValidBakery:=MaybeBakery?, MainInventoryUI:=ValidBakery.MaybeMainInventoryUI?):
                    spawn:
                        MainInventoryUI.DisplayDamageEffect()

    MakeRayPlayerDetectorSpheres(RayRadius : float, RayCentralPosition : vector3) : []agent =
        var AllAgentsInArea : []agent = array{}

        SpherePlayerDetectorRadius : float = 25.0

        #UpSpheres
        UpSphere0_ZPosition := RayCentralPosition.Z+RayRadius
        UpSphere1_ZPosition := UpSphere0_ZPosition+SpherePlayerDetectorRadius
        UpSphere2_ZPosition := UpSphere1_ZPosition+SpherePlayerDetectorRadius

        UpSphere0_Position : vector3 = vector3{X:=RayCentralPosition.X,Y:=RayCentralPosition.Y,Z:=UpSphere0_ZPosition}
        UpSphere1_Position : vector3 = vector3{X:=RayCentralPosition.X,Y:=RayCentralPosition.Y,Z:=UpSphere1_ZPosition}
        UpSphere2_Position : vector3 = vector3{X:=RayCentralPosition.X,Y:=RayCentralPosition.Y,Z:=UpSphere2_ZPosition}

        #DownSpheres
        DownSphere0_ZPosition := RayCentralPosition.Z-RayRadius
        DownSphere1_ZPosition := DownSphere0_ZPosition-SpherePlayerDetectorRadius
        DownSphere2_ZPosition := DownSphere1_ZPosition-SpherePlayerDetectorRadius

        DownSphere0_Position : vector3 = vector3{X:=RayCentralPosition.X,Y:=RayCentralPosition.Y,Z:=DownSphere0_ZPosition}
        DownSphere1_Position : vector3 = vector3{X:=RayCentralPosition.X,Y:=RayCentralPosition.Y,Z:=DownSphere1_ZPosition}
        DownSphere2_Position : vector3 = vector3{X:=RayCentralPosition.X,Y:=RayCentralPosition.Y,Z:=DownSphere2_ZPosition}

        AllPointSpheres : []vector3 = array{UpSphere0_Position,UpSphere1_Position,UpSphere2_Position,DownSphere0_Position,DownSphere1_Position,DownSphere2_Position}

        #for(Point : AllPointSpheres):
            #DebugDraw.DrawSphere(Point,?Radius:=SpherePlayerDetectorRadius,?Color:=NamedColors.Yellow,?Duration:=5.0)

        for(Player : GetPlayspace().GetPlayers(), Agent := agent[Player], PlayerCharacter := Player.GetFortCharacter[], PlayerTranslation := PlayerCharacter.GetTransform().Translation):
            if(1000.0>Distance(PlayerTranslation,RayCentralPosition)):
                var IsPlayingOnArea : logic = false
                #Only calcule if it is near 10 meters
                if(RayRadius>Distance(RayCentralPosition,PlayerTranslation)):
                    set AllAgentsInArea = AllAgentsInArea + array{Agent}
                else:
                    for(Index:=0..AllPointSpheres.Length-1, SpherePoint := AllPointSpheres[Index]):
                        if(SpherePlayerDetectorRadius>Distance(SpherePoint,PlayerTranslation)):
                            set AllAgentsInArea = AllAgentsInArea + array{Agent}
        return AllAgentsInArea


    MakeSpecialRay_DamageForward(NPCTransform : transform, NPCViewlocation : vector3, NPCViewRotation : rotation, NPCType : int, BaseDamage : float) : void =
        var AllPointSpheres : []vector3 = array{}
        var ViewLocation :vector3 = vector3{} 

        var Range : float = 10.0
        var Elevation : float = 0.0
        var Damage : float = 0.0
        var Duration : float = 0.0

        Log("Creating attack with damage {GetOptimizedMagnitudeLetter(BaseDamage)}",LOG_LEVEL_DEBUG7)

        if(NPCType=GROUND_MOB):
            set Range = SmallSpitRange
            set Elevation = -64.0
            set Damage = BaseDamage
            set Duration = SmallSpitDuration
        else if(NPCType=DRUID_MOB):
            set Range = SmallSpitRange
            set Elevation = -64.0
            set Damage = BaseDamage
            set Duration = SmallSpitDuration
        else if(NPCType=PLANT_MOB):
            set Range = LongSpitRange
            set Elevation = 49.0
            set Damage = BaseDamage
            set Duration = LongSpitDuration
        else if(NPCType=MINIBOSS_MOB):
            set Range = MiniBossSpitRange
            set Elevation = 0.0
            set Damage = BaseDamage
            set Duration = LongSpitDuration

        if(Range <= SmallSpitRange):
            set ViewLocation = NPCViewlocation+vector3{Z:=Elevation-16.0}
            EffectTransform:=transform{Translation:=NPCTransform.Translation+vector3{Z:=Elevation+10.0},Rotation:=NPCTransform.Rotation,Scale:=vector3{X:=1.0,Y:=1.0,Z:=1.0}}
            spawn:
                SpitLife(BP_SpitDamage_Short,EffectTransform)
        else:
            if(NPCType=PLANT_MOB):
                set ViewLocation = NPCViewlocation+vector3{Z:=Elevation-54.0}
            else if (NPCType=MINIBOSS_MOB):
                set ViewLocation = NPCViewlocation+vector3{Z:=Elevation-54.0}
            EffectTransform:=transform{Translation:=NPCTransform.Translation+vector3{Z:=Elevation},Rotation:=NPCTransform.Rotation,Scale:=vector3{X:=1.0,Y:=1.0,Z:=1.0}}
            spawn:
                SpitLife(BP_SpitDamage_Long,EffectTransform)

        SphereMarginRadius := 50.0
        SphereMargin := 32.0
        Sphere0_ToSphere1_Margin := 24.0

        Sphere0_PointRadius := SphereMarginRadius*0.25
        Sphere1_PointRadius := SphereMarginRadius*0.5
        Sphere2_PointRadius := SphereMarginRadius*0.75

        Sphere0_Point := ViewLocation+NPCViewRotation.GetLocalForward()*Sphere0_PointRadius
        Sphere1_Point := Sphere0_Point+NPCViewRotation.GetLocalForward()*Sphere0_ToSphere1_Margin
        Sphere2_Point := Sphere1_Point+NPCViewRotation.GetLocalForward()*SphereMargin
        Sphere3_Point := Sphere2_Point+NPCViewRotation.GetLocalForward()*SphereMargin

        set AllPointSpheres = array{Sphere0_Point,Sphere1_Point,Sphere2_Point,Sphere3_Point}

        ##Print("Sphere0_Point {Sphere0_Point}")

        BasicRange := Sphere0_PointRadius + Sphere0_ToSphere1_Margin + SphereMargin + SphereMargin + SphereMarginRadius

        #DebugDraw.DrawSphere(Sphere0_Point,?Radius:=Sphere0_PointRadius,?Color:=NamedColors.Red,?Duration:=5.0)
        #DebugDraw.DrawSphere(Sphere1_Point,?Radius:=Sphere1_PointRadius,?Color:=NamedColors.White,?Duration:=5.0)
        #DebugDraw.DrawSphere(Sphere2_Point,?Radius:=Sphere2_PointRadius,?Color:=NamedColors.Red,?Duration:=5.0)
        #DebugDraw.DrawSphere(Sphere3_Point,?Radius:=SphereMarginRadius,?Color:=NamedColors.White,?Duration:=5.0)

        if(Range>BasicRange):
            ##Print("Range is bigger, adding more spheres")
            if:
                NumSpheresToAdd := Ceil[(Range-BasicRange)/SphereMargin]
                NumSpheresToAdd>0
            then:
                ##Print("Trying to add {NumSpheresToAdd} Spheres")
                for(Index:=0..NumSpheresToAdd-1):
                    NewSphere_Point := Sphere3_Point+NPCViewRotation.GetLocalForward()*SphereMargin*(Index*1.0)
                    set AllPointSpheres = AllPointSpheres + array{NewSphere_Point}
                    #DebugDraw.DrawSphere(NewSphere_Point,?Radius:=SphereMarginRadius,?Color:=NamedColors.Blue,?Duration:=5.0)
        spawn:
            SpawnRayPersistence(Damage, Duration-0.75, AllPointSpheres, Sphere0_PointRadius, Sphere1_PointRadius, Sphere2_PointRadius, SphereMarginRadius)
        <#
        var AllAgentsToDamage : []agent = array{}
        var AllIsAgentAlreadyDamaged : [agent]logic = map{}

        for(Player : GetPlayspace().GetPlayers(), Agent := agent[Player], set AllIsAgentAlreadyDamaged[Agent] = false):
            
        for(Index:=0..AllPointSpheres.Length-1, SpherePoint := AllPointSpheres[Index]):
            if(Index=0):
                set AllAgentsToDamage = AllAgentsToDamage + MakeRayPlayerDetectorSpheres(Sphere0_PointRadius,SpherePoint)
            else if(Index=1):
                set AllAgentsToDamage = AllAgentsToDamage + MakeRayPlayerDetectorSpheres(Sphere1_PointRadius,SpherePoint)
            else if(Index=2):
                set AllAgentsToDamage = AllAgentsToDamage + MakeRayPlayerDetectorSpheres(Sphere2_PointRadius,SpherePoint)
            else:
                set AllAgentsToDamage = AllAgentsToDamage + MakeRayPlayerDetectorSpheres(SphereMarginRadius,SpherePoint)


        for(AgentToDamage : AllAgentsToDamage):
            if(IsAlreadyDamaged := AllIsAgentAlreadyDamaged[AgentToDamage], not IsAlreadyDamaged?, PlayerCharacter := AgentToDamage.GetFortCharacter[]):
                DamageCharacter(PlayerCharacter,Damage)
                if(set AllIsAgentAlreadyDamaged[AgentToDamage] = true):
                    #Print("Damaging player: {Damage}")#>

    
    SpawnRayPersistence(Damage : float, Duration : float, AllPointSpheres : []vector3, Sphere0_PointRadius : float, Sphere1_PointRadius : float, Sphere2_PointRadius : float, SphereMarginRadius : float)<suspends> : void =
        var AllIsAgentAlreadyDamaged : [agent]logic = map{}
        var TotalLife : float = 0.0
        for(Player : GetPlayspace().GetPlayers(), Agent := agent[Player], set AllIsAgentAlreadyDamaged[Agent] = false):
        loop:
            var AllAgentsToDamage : []agent = array{}
            for(Index:=0..AllPointSpheres.Length-1, SpherePoint := AllPointSpheres[Index]):
                if(Index=0):
                    set AllAgentsToDamage = AllAgentsToDamage + MakeRayPlayerDetectorSpheres(Sphere0_PointRadius,SpherePoint)
                else if(Index=1):
                    set AllAgentsToDamage = AllAgentsToDamage + MakeRayPlayerDetectorSpheres(Sphere1_PointRadius,SpherePoint)
                else if(Index=2):
                    set AllAgentsToDamage = AllAgentsToDamage + MakeRayPlayerDetectorSpheres(Sphere2_PointRadius,SpherePoint)
                else:
                    set AllAgentsToDamage = AllAgentsToDamage + MakeRayPlayerDetectorSpheres(SphereMarginRadius,SpherePoint)
            for(AgentToDamage : AllAgentsToDamage):
                if(IsAlreadyDamaged := AllIsAgentAlreadyDamaged[AgentToDamage], not IsAlreadyDamaged?, PlayerCharacter := AgentToDamage.GetFortCharacter[]):
                    DamageCharacter(PlayerCharacter,Damage)
                    if(set AllIsAgentAlreadyDamaged[AgentToDamage] = true):
                        ##Print("Damaging player: {Damage}")
            Sleep(0.04)
            set TotalLife = TotalLife + 0.04
            if(TotalLife>=Duration):
                break

    <#
    MakeSpecialRay_DamageForward(NPCharacter : fort_character, Damage : float, Range : float, Elevation : float) : void =
        var AllPointSpheres : []vector3 = array{}
        var ViewLocation :vector3 = vector3{} 

        if(Range <= DruidMobAttack_Range):
            set ViewLocation = NPCharacter.GetViewLocation()+vector3{Z:=Elevation-16.0}
            EffectTransform:=transform{Translation:=NPCharacter.GetTransform().Translation+vector3{Z:=Elevation-5.0},Rotation:=NPCharacter.GetTransform().Rotation,Scale:=vector3{X:=1.0,Y:=1.0,Z:=1.0}}
            spawn:
                SpitLife(BP_SpitDamage_Short,EffectTransform)
        else:
            set ViewLocation = NPCharacter.GetViewLocation()+vector3{Z:=Elevation-84.0}
            EffectTransform:=transform{Translation:=NPCharacter.GetTransform().Translation+vector3{Z:=Elevation},Rotation:=NPCharacter.GetTransform().Rotation,Scale:=vector3{X:=1.0,Y:=1.0,Z:=1.0}}
            spawn:
                SpitLife(BP_SpitDamage_Long,EffectTransform)

        ViewRotation:=NPCharacter.GetViewRotation()

        SphereMarginRadius := 50.0
        SphereMargin := 32.0
        Sphere0_ToSphere1_Margin := 24.0

        Sphere0_PointRadius := SphereMarginRadius*0.25
        Sphere1_PointRadius := SphereMarginRadius*0.5
        Sphere2_PointRadius := SphereMarginRadius*0.75

        Sphere0_Point := ViewLocation+ViewRotation.GetLocalForward()*Sphere0_PointRadius
        Sphere1_Point := Sphere0_Point+ViewRotation.GetLocalForward()*Sphere0_ToSphere1_Margin
        Sphere2_Point := Sphere1_Point+ViewRotation.GetLocalForward()*SphereMargin
        Sphere3_Point := Sphere2_Point+ViewRotation.GetLocalForward()*SphereMargin

        set AllPointSpheres = array{Sphere0_Point,Sphere1_Point,Sphere2_Point,Sphere3_Point}

        #Print("Sphere0_Point {Sphere0_Point}")

        BasicRange := Sphere0_PointRadius + Sphere0_ToSphere1_Margin + SphereMargin + SphereMargin + SphereMarginRadius

        DebugDraw.DrawSphere(Sphere0_Point,?Radius:=Sphere0_PointRadius,?Color:=NamedColors.Red,?Duration:=5.0)
        DebugDraw.DrawSphere(Sphere1_Point,?Radius:=Sphere1_PointRadius,?Color:=NamedColors.White,?Duration:=5.0)
        DebugDraw.DrawSphere(Sphere2_Point,?Radius:=Sphere2_PointRadius,?Color:=NamedColors.Red,?Duration:=5.0)
        DebugDraw.DrawSphere(Sphere3_Point,?Radius:=SphereMarginRadius,?Color:=NamedColors.White,?Duration:=5.0)

        if(Range>BasicRange):
            #Print("Range is bigger, adding more spheres")
            if:
                NumSpheresToAdd := Ceil[(Range-BasicRange)/SphereMargin]
                NumSpheresToAdd>0
            then:
                #Print("Trying to add {NumSpheresToAdd} Spheres")
                for(Index:=0..NumSpheresToAdd-1):
                    NewSphere_Point := Sphere3_Point+ViewRotation.GetLocalForward()*SphereMargin*(Index*1.0)
                    set AllPointSpheres = AllPointSpheres + array{NewSphere_Point}
                    DebugDraw.DrawSphere(NewSphere_Point,?Radius:=SphereMarginRadius,?Color:=NamedColors.Blue,?Duration:=5.0)

        for(Player : GetPlayspace().GetPlayers(), PlayerCharacter := Player.GetFortCharacter[], PlayerTranslation := PlayerCharacter.GetTransform().Translation):
            if(1000.0>Distance(PlayerTranslation,NPCharacter.GetTransform().Translation)):
                var IsPlayingOnArea : logic = false
                #Only calcule if it is near 10 meters
                for(Index:=0..AllPointSpheres.Length-1, SpherePoint := AllPointSpheres[Index]):
                    if(Index=0):
                        if(Sphere0_PointRadius>Distance(SpherePoint,PlayerTranslation)):
                            set IsPlayingOnArea = true
                    else if(Index=1):
                        if(Sphere1_PointRadius>Distance(SpherePoint,PlayerTranslation)):
                            set IsPlayingOnArea = true
                    else if(Index=2):
                        if(Sphere2_PointRadius>Distance(SpherePoint,PlayerTranslation)):
                            set IsPlayingOnArea = true
                    else:
                        if(SphereMarginRadius>Distance(SpherePoint,PlayerTranslation)):
                            set IsPlayingOnArea = true
                if(IsPlayingOnArea?):
                    DamageCharacter(PlayerCharacter,Damage)#>

    MakeDamageForwardOld(NPCharacter : fort_character, Damage : float, AmountForward : float) : void =
        ViewLocation:=NPCharacter.GetViewLocation()
        ViewRotation:=NPCharacter.GetViewRotation()

        PointingVector:=ViewRotation.GetLocalForward()*AmountForward
        HalfPointingVector:=ViewRotation.GetLocalForward()*(AmountForward/2.0)

        NewPosition:=ViewLocation+PointingVector
        NewPositionHalf:=ViewLocation+HalfPointingVector

        #DoubleArea
        for(Player : GetPlayspace().GetPlayers(), PlayerCharacter := Player.GetFortCharacter[]):
            var IsPlayingOnArea : logic = false
            if((AmountForward/2.0)>DistanceXY(NPCharacter.GetTransform().Translation,NewPosition)):
                set IsPlayingOnArea = true
            if((AmountForward/2.0)>DistanceXY(NPCharacter.GetTransform().Translation,NewPositionHalf)):
                set IsPlayingOnArea = true
            if(IsPlayingOnArea?):
                DamageCharacter(PlayerCharacter,Damage)

    
    MakeMovementDecision(NPCAgent : agent,NPCCharacter : fort_character, MobType : int, NPCSpawnPoint : transform) : tuple(?agent,?vector3,?vector3,logic,float,float) =
        var Mob_MoveToWait_Duration : float = 1.0
        var MaybeAgentObjetive : ?agent = false
        var MaybePrimary_TargetPosition : ?vector3 = false
        var MaybeSecondary_TargetPosition : ?vector3 = false
        
        var IsClippingDetected : logic = false
        
        var ReachRadius : float = 64.0
        var MoveToWaitDuration : float = 1.0

        if(MobType=MINIBOSS_MOB):
            set ReachRadius = MiniBossAttack_Range
            set MoveToWaitDuration = MiniBoss_MoveToWait_Duration
            if(NewTargetCharacter:=FindNearestTargetFromPos[NPCCharacter,Mob_EnemyRangeDetection], TargetAgent := NewTargetCharacter.GetAgent[]):
                set MaybeAgentObjetive = option{TargetAgent}
            Primary_TargetPosition := FindRandomNavigationPoint(NPCCharacter.GetTransform().Translation,MiniBoss_Max_Movement,NPCSpawnPoint.Translation)
            set MaybePrimary_TargetPosition = option{Primary_TargetPosition}
            set MaybeSecondary_TargetPosition = option{Primary_TargetPosition}
            if(10.0<Abs(Primary_TargetPosition.Z-NPCSpawnPoint.Translation.Z)):
                set IsClippingDetected = true
                #Print("Detected clipping error")
        else if(MobType=PLANT_MOB):
            set Mob_MoveToWait_Duration = PlantMob_MoveToWait_Duration
            if(NewTargetCharacter:=FindNearestTargetFromPos[NPCCharacter,Mob_EnemyRangeDetection], TargetAgent := NewTargetCharacter.GetAgent[]):
                #Find enemy
                set MaybeAgentObjetive = option{TargetAgent}
                #The plant does not move
                set MaybePrimary_TargetPosition = option{NPCSpawnPoint.Translation}
                set MaybeSecondary_TargetPosition = option{NPCSpawnPoint.Translation}

        else if(MobType=DRUID_MOB):
            set ReachRadius = DruidMobAttack_Range-56.0
            set MoveToWaitDuration = DruidMob_MoveToWait_Duration
            if(NewTargetCharacter:=FindNearestTargetFromPos[NPCCharacter,Mob_EnemyRangeDetection], TargetAgent := NewTargetCharacter.GetAgent[]):
                #Find enemy
                set MaybeAgentObjetive = option{TargetAgent}
                set MaybePrimary_TargetPosition = option{NewTargetCharacter.GetTransform().Translation}
                #Secondary target
                Secondary_TargetPosition := FindRandomNavigationPoint(NPCCharacter.GetTransform().Translation, GroundMob_Max_Movement,NPCSpawnPoint.Translation)
                set MaybeSecondary_TargetPosition = option{Secondary_TargetPosition}
            else:
                #Creation two posible positions to move
                Primary_TargetPosition := FindRandomNavigationPoint(NPCCharacter.GetTransform().Translation, GroundMob_Max_Movement,NPCSpawnPoint.Translation)
                set MaybePrimary_TargetPosition = option{Primary_TargetPosition}
                Secondary_TargetPosition := FindRandomNavigationPoint(NPCCharacter.GetTransform().Translation, GroundMob_Max_Movement,NPCSpawnPoint.Translation)
                set MaybeSecondary_TargetPosition = option{Secondary_TargetPosition}

        else if(MobType=GROUND_MOB):
            set ReachRadius = GroundMobAttack_Range-56.0
            set MoveToWaitDuration = GroundMob_MoveToWait_Duration
            if(NewTargetCharacter:=FindNearestTargetFromPos[NPCCharacter,Mob_EnemyRangeDetection], TargetAgent := NewTargetCharacter.GetAgent[]):
                #Find enemy
                set MaybeAgentObjetive = option{TargetAgent}
                set MaybePrimary_TargetPosition = option{NewTargetCharacter.GetTransform().Translation}
                #Secondary target
                Secondary_TargetPosition := FindRandomNavigationPoint(NPCCharacter.GetTransform().Translation, GroundMob_Max_Movement,NPCSpawnPoint.Translation)
                set MaybeSecondary_TargetPosition = option{Secondary_TargetPosition}
            else:
                #Creation two posible positions to move
                Primary_TargetPosition := FindRandomNavigationPoint(NPCCharacter.GetTransform().Translation, GroundMob_Max_Movement,NPCSpawnPoint.Translation)
                set MaybePrimary_TargetPosition = option{Primary_TargetPosition}
                Secondary_TargetPosition := FindRandomNavigationPoint(NPCCharacter.GetTransform().Translation, GroundMob_Max_Movement,NPCSpawnPoint.Translation)
                set MaybeSecondary_TargetPosition = option{Secondary_TargetPosition}


        return (MaybeAgentObjetive,MaybePrimary_TargetPosition,MaybeSecondary_TargetPosition,IsClippingDetected, Mob_NavigationErrorWait,ReachRadius)
    
        <#
        #Making movement
        if(TargetAgent:=MaybeAgentObjetive?):
            ##Print("Changing focus")
            spawn{NPCFocus.MaintainFocus(TargetAgent)}
        if:
            Primary_TargetPosition := MaybePrimary_TargetPosition?
            Secondary_TargetPosition := MaybeSecondary_TargetPosition?
        then:
            Primary_NPCNavigationTarget := MakeNavigationTarget(vector3{X:=Primary_TargetPosition.X,Y:=Primary_TargetPosition.Y,Z:=Primary_TargetPosition.Z})
            NavResultGoTo1:=NPCNavigatable.NavigateTo(Primary_NPCNavigationTarget, ?MovementType := movement_types.Walking, ?ReachRadius := ReachRadius)
            if(ClippingDetected?):
                PlayDirtEffect(NPCCharacter.GetTransform())
                if(MobType=MINIBOSS_MOB,NPCCharacter.TeleportTo[NPCSpawnPoint.Translation, NPCSpawnPoint.Rotation]):
                    PlayDirtEffect(NPCSpawnPoint)
                    #Print("Teleporting miniboss mob to a acceptable position")
                else:
                    set IsNavigationError = true
                    Sleep(Mob_NavigationErrorWait)
            else:
                if(NavResultGoTo1 <> navigation_result.Reached):
                    #try to navigate to point 2
                    Secondary_NPCNavigationTarget := MakeNavigationTarget(vector3{X:=Secondary_TargetPosition.X,Y:=Secondary_TargetPosition.Y,Z:=Secondary_TargetPosition.Z})
                    NavResultGoTo2:=NPCNavigatable.NavigateTo(Primary_NPCNavigationTarget, ?MovementType := movement_types.Walking, ?ReachRadius := ReachRadius) 
                    if(NavResultGoTo2 <> navigation_result.Reached):
                        #If there is again an error, teleport to spawn point
                        PlayDirtEffect(NPCCharacter.GetTransform())
                        if(NPCCharacter.TeleportTo[NPCSpawnPoint.Translation, NPCSpawnPoint.Rotation]):
                            PlayDirtEffect(NPCSpawnPoint)
                        else:
                            #Can not teleport? make an error
                            set IsNavigationError = true
                            Sleep(Mob_NavigationErrorWait)
                            
                        
        else:
            NPCNavigatable.Wait(?Duration := MoveToWaitDuration)#>
        
    
    SpawnMiniBossMinions(NPCTransform : transform, SpawnOrigin : vector3)<suspends> : void =
        if(GeneralMobsManager := MaybeGeneralMobsManager?):
            NumDungeon := GeneralMobsManager.FindNearDungeon(NPCTransform.Translation)
            SpawnPosition := FindRandomNavigationPoint(NPCTransform.Translation,512.0,NPCTransform.Translation)
            MobsSpawnTransform := transform{Translation:=vector3{X:=SpawnPosition.X,Y:=SpawnPosition.Y,Z:=SpawnOrigin.Z},Rotation:=NPCTransform.Rotation,Scale:=vector3{X:=1.0,Y:=1.0,Z:=1.0}}
            TupleMobLevel:=GeneralMobsManager.DungeonDataBase.GetMobsLV(NumDungeon)
            NumMobsToSpawn:=GetRandomInt(2,4)
            for(Index:=0..NumMobsToSpawn-1):
                if(Index=0,NumMobsToSpawn>2):
                    GeneralMobsManager.ExecuteDungeon_MobSpawn(NumDungeon,MobsSpawnTransform,false,DRUID_MOB,TupleMobLevel(1),true)
                else:
                    GeneralMobsManager.ExecuteDungeon_MobSpawn(NumDungeon,MobsSpawnTransform,false,GROUND_MOB,TupleMobLevel(1),true)
                Sleep(0.25)

    FindRandomNavigationPoint(CurrentPosition : vector3, MaxDispersion : float, Origin : vector3) : vector3 =
        var NewPoint : vector3 = Origin
        RandomXFloat := GetRandomFloat(-MaxDispersion/2.0,MaxDispersion/2.0)
        RandomYFloat := GetRandomFloat(-MaxDispersion/2.0,MaxDispersion/2.0)

        set NewPoint.X = NewPoint.X + RandomXFloat
        set NewPoint.Y = NewPoint.Y + RandomYFloat
        if(Distance(Origin,NewPoint)>MaxDispersion):
            return Origin
        else:
            return NewPoint 

    FindNearestTargetFromPos(FortChar : fort_character, InCheckRange : float)<decides>< transacts> : fort_character = 
        var MaybeTarget : ?fort_character = false

        #Change this range how you would like
        #This will be the range that the device will search for the nearest target
        var CheckRange : float = InCheckRange

        for (Player : GetPlayspace().GetPlayers(), PlayerFortChar := Player.GetFortCharacter[]) {
            DistanceDifference := Distance(FortChar.GetTransform().Translation, PlayerFortChar.GetTransform().Translation)

            if (DistanceDifference < CheckRange and not PlayerFortChar = FortChar) {
                set MaybeTarget = option{PlayerFortChar}
                set CheckRange = DistanceDifference 
            }
        }

        return MaybeTarget?
    

    SpawnMobHealZone(Transform : transform)<suspends> : ?creative_prop =
        SpawnTransform:=transform{Translation:=Transform.Translation+vector3{Z:=-74.0},Rotation:=Transform.Rotation,Scale:=Transform.Scale}
        TupleSpawn:=SpawnProp(BP_MobHealZone,SpawnTransform)
        MaybeCreativeProp:=TupleSpawn(0)
        if(ValidProp := MaybeCreativeProp?):
            spawn:
                DisposeEffect(ValidProp,5.0)
        return MaybeCreativeProp

    SpawnDamageZone(Transform : transform): ?creative_prop =
        
        SpawnTransform:=transform{Translation:=Transform.Translation+vector3{Z:=-80.0},Rotation:=Transform.Rotation,Scale:=Transform.Scale}
        TupleSpawn:=SpawnProp(BP_DamageZone,SpawnTransform)
        MaybeCreativeProp:=TupleSpawn(0)
        if(ValidProp := MaybeCreativeProp?):
            spawn:
                DisposeEffect(ValidProp,5.0)
        return MaybeCreativeProp

        <#
    SpawnSpitDamage(Transform : transform, Objetive : fort_character, Elevation : float, IsShort : logic) : void =

        FinalTranslation:=transform{Translation:=Transform.Translation+vector3{Z:=Elevation},Rotation:=Transform.Rotation,Scale:=Transform.Scale}
        if(IsShort?):
            spawn:
                SpitLife(BP_SpitDamage_Short,FinalTranslation)
        else:
            spawn:
                SpitLife(BP_SpitDamage_Long,FinalTranslation)#>

    SpitLife(BP_SpitDamage : creative_prop_asset, OriginTransform : transform)<suspends> : void =
        ##Print("Shooting projectile")

        #Rotation:=GetForwardRotation_Vector3(Objetive.GetTransform().Translation,OriginTranslation)
        SpawnTransform:=transform{Translation := OriginTransform.Translation, Rotation := OriginTransform.Rotation, Scale := vector3{X:=1.0,Y:=1.0,Z:=1.0}}

        TupleSpawn:=SpawnProp(BP_SpitDamage,SpawnTransform)
        MaybeCreativeProp:=TupleSpawn(0)
        if(ValidProp := MaybeCreativeProp?):
            Sleep(3.0)
            ValidProp.Dispose()


    <#SpawnProjectile(Transform : transform, Objetive : fort_character) : void =
        spawn:
            ProjectileLife(Transform,Objetive)#>

        
    <#ProjectileLife(OriginTransform : transform, Objetive : fort_character)<suspends> : void=
        ##Print("Shooting projectile")

        TotalDistance := Distance(OriginTransform.Translation,Objetive.GetTransform().Translation)

        Time:=TotalDistance/10000.0

        Rotation:=GetForwardRotation_Vector3(Objetive.GetTransform().Translation,OriginTransform.Translation)

        ObjetiveTraslation := Objetive.GetTransform().Translation
        FinalTransform:=transform{Translation := ObjetiveTraslation, Rotation := Rotation, Scale := OriginTransform.Scale}
        SpawnTransform:=transform{Translation := OriginTransform.Translation, Rotation := Rotation, Scale := OriginTransform.Scale}

        TupleSpawn:=SpawnProp(BP_Projectile,SpawnTransform)
        MaybeCreativeProp:=TupleSpawn(0)
        if(ValidProp := MaybeCreativeProp?, ValidProp.IsValid[]):
            ValidProp.MoveTo(FinalTransform,Time)
            Sleep(Time)
            ValidProp.Dispose()#>

    SetGeneralMobManager(GeneralMobsManager : general_mobs_manager) : void =
        set MaybeGeneralMobsManager = option{GeneralMobsManager}

    # Runs when the device is started in a running game
    OnBegin<override>()<suspends>:void=

        for(AllAudioPlayer : AllAudioPlayer_BossBiteSound):
            set AllAudioPlayerBossBite_Used = AllAudioPlayerBossBite_Used + array{false}

        for(AllAudioPlayer : AllAudioPlayer_LoadBossBiteSound):
            set AllAudioPlayerLoadBossBite_Used = AllAudioPlayerLoadBossBite_Used + array{false}

        for(AllAudioPlayer : AllAudioPlayer_Bite0Sound):
            set AllAudioPlayerBite0_Used = AllAudioPlayerBite0_Used + array{false}

        for(AllAudioPlayer : AllAudioPlayer_Bite1Sound):
            set AllAudioPlayerBite1_Used = AllAudioPlayerBite1_Used + array{false}

        for(AllAudioPlayer : AllAudioPlayer_LoadBite1Sound):
            set AllAudioPlayerLoadBite1_Used = AllAudioPlayerLoadBite1_Used + array{false}
    
        for(AllAudioPlayer : AllAudioPlayer_LoadBite0Sound):
            set AllAudioPlayerLoadBite0_Used = AllAudioPlayerLoadBite0_Used + array{false}

        for(AllAudioPlayer : AllAudioPlayer_DirtSound):
            set AllAudioPlayerDirt_Used = AllAudioPlayerDirt_Used + array{false}

        for(AllAudioPlayer : AllAudioPlayer_DeathSound):
            set AllAudioPlayerDeath_Used = AllAudioPlayerDeath_Used + array{false}

        for(AllAudioPlayer : AllAudioPlayer_LoadZone):
            set AllAudioLoadZone_Used = AllAudioLoadZone_Used + array{false}

        for(AllAudioPlayer : AllAudioPlayer_DamageZone):
            set AllAudioDamageZone_Used = AllAudioDamageZone_Used + array{false}

        for(AllAudioPlayer : AllAudioPlayer_BossLaugh):
            set AllAudioPlayerBossLaugh_Used = AllAudioPlayerBossLaugh_Used + array{false}

        for(AllAudioPlayer : AllAudioPlayer_MobHealerZone):
            set AllAudioMobHealerZone_Used = AllAudioMobHealerZone_Used + array{false}

        loop:
            Sleep(1.0)
            set TotalTime = TotalTime + 1

            #for(Player : GetPlayspace().GetPlayers(), Character := Player.GetFortCharacter[]):
            #    LowSpherePos := vector3{X:=Character.GetTransform().Translation.X,Y:=Character.GetTransform().Translation.Y,Z:=Character.GetTransform().Translation.Z-50.0}
            #    HighSpherePos := vector3{X:=Character.GetTransform().Translation.X,Y:=Character.GetTransform().Translation.Y,Z:=Character.GetTransform().Translation.Z+50.0}

                #DebugDraw.DrawSphere(LowSpherePos,?Radius:=25.0,?Color:=NamedColors.White,?Duration:=5.0)
                #DebugDraw.DrawSphere(Character.GetTransform().Translation,?Radius:=25.0,?Color:=NamedColors.Red,?Duration:=5.0)
                #DebugDraw.DrawSphere(HighSpherePos,?Radius:=25.0,?Color:=NamedColors.Blue,?Duration:=5.0)


    
    PlayDeathEffect(MobType : int, CharacterTransform : transform) : void =
        ##Print("Playing death effect")
        spawn:
            PlaySound(CharacterTransform, SOUND_DEATH,-1.0)
        if(MobType=MINIBOSS_MOB):
            TupleSpawn:=SpawnProp(MiniBossDeathEffect,CharacterTransform)
            MaybeCreativeProp:=TupleSpawn(0)
            if(ValidProp := MaybeCreativeProp?):
                spawn:
                    DisposeEffect(ValidProp,5.0)
            else:
                PrintFatalError("PlayDeathEffect creative prop spawned not valid")
        else:
            TupleSpawn:=SpawnProp(MobDeathEffect,CharacterTransform)
            MaybeCreativeProp:=TupleSpawn(0)
            if(ValidProp := MaybeCreativeProp?):
                spawn:
                    DisposeEffect(ValidProp,5.0)
            else:
                PrintFatalError("PlayDeathEffect creative prop spawned not valid")

    PlayDirtEffect(Transform : transform) : void =
        ##Print("Playing dirt effect")
        spawn:
            PlaySound(Transform, SOUND_DIRT,-1.0)
        if(GeneralMobsManager:=MaybeGeneralMobsManager?):
            TupleSpawn:=SpawnProp(GeneralMobsManager.DirtSpawnEffect,Transform)
            MaybeCreativeProp:=TupleSpawn(0)
            if(ValidProp := MaybeCreativeProp?):
                spawn:
                    DisposeEffect(ValidProp,5.0)
            else:
                PrintFatalError("PlayDeathEffect creative prop spawned not valid")

    HandlerPlayBite(MobType: int, CharacterTransform : transform, PreDelay : float) : void =
        CacheTransform:=CharacterTransform
        CacheType:=MobType
        CachePreDelay:=PreDelay
        spawn:
            PlayBite(CacheType, CacheTransform, CachePreDelay)

    PlayBite<private>(MobType: int, CharacterTransform : transform, PreDelay : float)<suspends> : void =
        var AudioPlayerLoad_ToUse : audio_player_device = audio_player_device{}
        var AudioPlayerBite_ToUse : audio_player_device = audio_player_device{}

        var IndexAudioLoad : int = -1
        var IndexAudioBite : int = -1
        if(MobType=GROUND_MOB or MobType=DRUID_MOB):
            #Load
            loop:
                if(IsUsed:=AllAudioPlayerLoadBite0_Used[IndexAudioLoad], not IsUsed?, set AllAudioPlayerLoadBite0_Used[IndexAudioLoad] = true):
                    if(AudioPlayer:=AllAudioPlayer_LoadBite0Sound[IndexAudioLoad]):
                        set AudioPlayerLoad_ToUse=AudioPlayer
                    break
                if(IndexAudioLoad>AllAudioPlayerLoadBite0_Used.Length-1):
                    PrintFatalError("Audio player not found")
                    break
                set IndexAudioLoad = IndexAudioLoad + 1
            #Bite
            loop:
                if(IsUsed:=AllAudioPlayerBite0_Used[IndexAudioBite], not IsUsed?, set AllAudioPlayerBite0_Used[IndexAudioBite] = true):
                    if(AudioPlayer:=AllAudioPlayer_Bite0Sound[IndexAudioBite]):
                        set AudioPlayerBite_ToUse=AudioPlayer
                    break
                if(IndexAudioBite>AllAudioPlayerBite0_Used.Length-1):
                    PrintFatalError("Audio player not found")
                    break
                set IndexAudioBite = IndexAudioBite + 1
        else if(MobType=PLANT_MOB):
            #Load
            loop:
                if(IsUsed:=AllAudioPlayerLoadBite1_Used[IndexAudioLoad], not IsUsed?, set AllAudioPlayerLoadBite1_Used[IndexAudioLoad] = true):
                    if(AudioPlayer:=AllAudioPlayer_LoadBite1Sound[IndexAudioLoad]):
                        set AudioPlayerLoad_ToUse=AudioPlayer
                    break
                if(IndexAudioLoad>AllAudioPlayerLoadBite1_Used.Length-1):
                    PrintFatalError("Audio player not found")
                    break
                set IndexAudioLoad = IndexAudioLoad + 1
            #Bite
            loop:
                if(IsUsed:=AllAudioPlayerBite1_Used[IndexAudioBite], not IsUsed?, set AllAudioPlayerBite1_Used[IndexAudioBite] = true):
                    if(AudioPlayer:=AllAudioPlayer_Bite1Sound[IndexAudioBite]):
                        set AudioPlayerBite_ToUse=AudioPlayer
                    break
                if(IndexAudioBite>AllAudioPlayerBite1_Used.Length-1):
                    PrintFatalError("Audio player not found")
                    break
                set IndexAudioBite = IndexAudioBite + 1
        else if(MobType=MINIBOSS_MOB):
            #Load
            loop:
                if(IsUsed:=AllAudioPlayerLoadBossBite_Used[IndexAudioLoad], not IsUsed?, set AllAudioPlayerLoadBossBite_Used[IndexAudioLoad] = true):
                    if(AudioPlayer:=AllAudioPlayer_LoadBossBiteSound[IndexAudioLoad]):
                        set AudioPlayerLoad_ToUse=AudioPlayer
                    break
                if(IndexAudioLoad>AllAudioPlayerLoadBossBite_Used.Length-1):
                    PrintFatalError("Audio player not found")
                    break
                set IndexAudioLoad = IndexAudioLoad + 1
            #Bite
            loop:
                if(IsUsed:=AllAudioPlayerBossBite_Used[IndexAudioBite], not IsUsed?, set AllAudioPlayerBossBite_Used[IndexAudioBite] = true):
                    if(AudioPlayer:=AllAudioPlayer_BossBiteSound[IndexAudioBite]):
                        set AudioPlayerBite_ToUse=AudioPlayer
                    break
                if(IndexAudioBite>AllAudioPlayerBossBite_Used.Length-1):
                    PrintFatalError("Audio player not found")
                    break
                set IndexAudioBite = IndexAudioBite + 1

        if(AudioPlayerLoad_ToUse.TeleportTo[CharacterTransform], AudioPlayerBite_ToUse.TeleportTo[CharacterTransform]):
            Sleep(0.1)
            AudioPlayerLoad_ToUse.Play()
            Sleep(PreDelay)
            AudioPlayerLoad_ToUse.Stop()
            AudioPlayerBite_ToUse.Play()
            Sleep(2.0)

        #Clean
        if(MobType=GROUND_MOB or MobType=DRUID_MOB):
            if(set AllAudioPlayerLoadBite0_Used[IndexAudioLoad]=false):
            if(set AllAudioPlayerBite0_Used[IndexAudioBite]=false):
        if(MobType=PLANT_MOB):
            if(set AllAudioPlayerLoadBite1_Used[IndexAudioLoad]=false):
            if(set AllAudioPlayerBite1_Used[IndexAudioBite]=false):
        if(MobType=MINIBOSS_MOB):
            if(set AllAudioPlayerLoadBossBite_Used[IndexAudioLoad]=false):
            if(set AllAudioPlayerBossBite_Used[IndexAudioBite]=false):
            


    DisposeEffect(Prop : creative_prop, Duration : float)<suspends> : void =
        Sleep(Duration)
        if(Prop.IsValid[]):
            Prop.Dispose()
    PlaySound(CharacterTransform : transform, SoundType : int, TimeLength : float)<suspends> : void =
        if(SoundType=SOUND_DIRT):
            ##Print("Playing DIRT sound!")
            var Index : int = -1
            loop:
                if(IsUsed:=AllAudioPlayerDirt_Used[Index], not IsUsed?, set AllAudioPlayerDirt_Used[Index] = true):
                    if(AudioPlayer:=AllAudioPlayer_DirtSound[Index]):
                        if(AudioPlayer.TeleportTo[CharacterTransform]):
                            Sleep(0.1)
                            AudioPlayer.Play()
                    Sleep(2.0)
                    if(set AllAudioPlayerDirt_Used[Index]=false):
                    break
                if(Index>AllAudioPlayerDirt_Used.Length-1):
                    PrintFatalError("Audio player not found")
                    break
                set Index = Index + 1
        else if(SoundType=SOUND_DEATH):
            ##Print("Playing DEATH sound!")
            var Index : int = -1
            loop:
                if(IsUsed:=AllAudioPlayerDeath_Used[Index], not IsUsed?, set AllAudioPlayerDeath_Used[Index] = true):
                    if(AudioPlayer:=AllAudioPlayer_DeathSound[Index]):
                        if(AudioPlayer.TeleportTo[CharacterTransform]):
                            Sleep(0.1)
                            AudioPlayer.Play()
                    Sleep(2.0)
                    if(set AllAudioPlayerDeath_Used[Index]=false):
                    break
                if(Index>AllAudioPlayerDeath_Used.Length-1):
                    PrintFatalError("Audio player not found")
                    break
                set Index = Index + 1
        else if(SoundType=SOUND_LOAD_ZONE):
            ##Print("Playing DEATH sound!")
            var Index : int = -1
            loop:
                if(IsUsed:=AllAudioLoadZone_Used[Index], not IsUsed?, set AllAudioLoadZone_Used[Index] = true):
                    if(AudioPlayer:=AllAudioPlayer_LoadZone[Index]):
                        if(AudioPlayer.TeleportTo[CharacterTransform]):
                            Sleep(0.1)
                            AudioPlayer.Play()
                            Sleep(TimeLength)
                            AudioPlayer.Stop()
                    Sleep(2.0)
                    if(set AllAudioLoadZone_Used[Index]=false):
                    break
                if(Index>AllAudioLoadZone_Used.Length-1):
                    PrintFatalError("Audio player not found")
                    break
                set Index = Index + 1
        else if(SoundType=SOUND_DAMAGE_ZONE):
            ##Print("Playing DEATH sound!")
            var Index : int = -1
            loop:
                if(IsUsed:=AllAudioDamageZone_Used[Index], not IsUsed?, set AllAudioDamageZone_Used[Index] = true):
                    if(AudioPlayer:=AllAudioPlayer_DamageZone[Index]):
                        if(AudioPlayer.TeleportTo[CharacterTransform]):
                            Sleep(0.1)
                            AudioPlayer.Play()
                    Sleep(2.0)
                    if(set AllAudioDamageZone_Used[Index]=false):
                    break
                if(Index>AllAudioDamageZone_Used.Length-1):
                    PrintFatalError("Audio player not found")
                    break
                set Index = Index + 1

        else if(SoundType=SOUND_HEAL_ZONE):
            var Index : int = -1
            loop:
                if(IsUsed:=AllAudioMobHealerZone_Used[Index], not IsUsed?, set AllAudioMobHealerZone_Used[Index] = true):
                    if(AudioPlayer:=AllAudioPlayer_MobHealerZone[Index]):
                        if(AudioPlayer.TeleportTo[CharacterTransform]):
                            Sleep(0.1)
                            AudioPlayer.Play()
                    Sleep(2.0)
                    if(set AllAudioMobHealerZone_Used[Index]=false):
                    break
                if(Index>AllAudioMobHealerZone_Used.Length-1):
                    PrintFatalError("Audio player not found")
                    break
                set Index = Index + 1

        else if(SoundType=SOUND_BOSS_LAUGH):
            var Index : int = -1
            loop:
                if(IsUsed:=AllAudioPlayerBossLaugh_Used[Index], not IsUsed?, set AllAudioPlayerBossLaugh_Used[Index] = true):
                    if(AudioPlayer:=AllAudioPlayer_BossLaugh[Index]):
                        if(AudioPlayer.TeleportTo[CharacterTransform]):
                            Sleep(0.1)
                            AudioPlayer.Play()
                    Sleep(2.0)
                    if(set AllAudioPlayerBossLaugh_Used[Index]=false):
                    break
                if(Index>AllAudioPlayerBossLaugh_Used.Length-1):
                    PrintFatalError("Audio player not found")
                    break
                set Index = Index + 1
        
            
