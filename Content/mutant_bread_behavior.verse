
using { /Fortnite.com/Devices }
using { /Verse.org/Simulation }
using { /UnrealEngine.com/Temporary/Diagnostics }

# See https://dev.epicgames.com/documentation/en-us/uefn/create-your-own-device-in-verse for how to create a verse device.

# A Verse-authored creative device that can be placed in a level
using { /Fortnite.com/AI }
using { /Verse.org/Simulation }
using { /Fortnite.com/Characters }
using { /Fortnite.com/Animation/PlayAnimation }
using { /Verse.org/Assets }
using { /UnrealEngine.com/Temporary/SpatialMath }
using { /Fortnite.com/Game }
using { /UnrealEngine.com/Temporary/UI }
using { /Verse.org/Colors/NamedColors }
 
# Getting started: https://www.epicgames.com/fortnite/en-US/creative/docs/uefn/Verse/onboarding-guide-to-programming-with-verse-in-unreal-editor-for-fortnite

# A Verse-authored NPC Behavior that can be used within an NPC Character Definition or an NPC Spawner device's NPC Behavior Script Override.


miniboss_behavior := class(npc_behavior):
    BehaviorHelp : skeleton_behavior_helper = skeleton_behavior_helper{}

    var Attack1 : animation_sequence = NPCsAnimations.MiniBoss_Attack_Anim
    #var Move : animation_sequence = NPCsAnimations.MiniBoss_Movement_Anim
    
    NPCDamage : float = 10.0

    # This function runs when the NPC is spawned in the world and ready to follow a behavior.
    OnBegin< override>()< suspends>:void=
        # TODO: Replace this with your code
        if: 
            
            NPCAgent := GetAgent[]
            NPCChar := NPCAgent.GetFortCharacter[]
            
            #Allows us to tell npcs to go to different place like the pathfinding service in roblox
            NPCNav := NPCChar.GetNavigatable[]

            #Control where npc is looking
            NPCFocus := NPCChar.GetFocusInterface[]
            

            #Control the npc's animations
            NPCAnim := NPCChar.GetPlayAnimationController[]
        then:
            NPCChar.DamagedEvent().Subscribe(NPCDamaged)
            NPCSpawnLocation := NPCChar.GetTransform().Translation

            loop:
                Sleep(0.1)
                #Checks if npc is alive
                if (not NPCChar.IsActive[]):
                    break
                
                if (NewTarget := BehaviorHelp.FindNearestTargetFromPos[NPCChar], NewAgent := NewTarget.GetAgent[]):
                    spawn{NPCFocus.MaintainFocus(NewAgent)}

                    if(AgentPosition := player[NewAgent].GetFortCharacter[].GetTransform().Translation):
                        NavTarget := MakeNavigationTarget(vector3{X:=AgentPosition.X,Y:=AgentPosition.Y,Z:=AgentPosition.Z})
                        #NavTarget := MakeNavigationTarget(NewAgent)
                        NPCNav.NavigateTo(NavTarget, ?MovementType := movement_types.Running, ?ReachRadius := 150.0)
                        NPCAnim.Play(Attack1,?BlendInTime:=0.25,?BlendOutTime:=0.25)
                        Sleep(0.58) #Time that least to attack
                    DistanceDifference := Distance(NPCChar.GetTransform().Translation, NewTarget.GetTransform().Translation)
                    if (DistanceDifference < 200.0):
                        NewTarget.Damage(NPCDamage)
                    Sleep(0.58) 
                    Sleep(1.00) #Time that least to attack
                
            
        

    # This function runs when the NPC is despawned wdor eliminated from the world.
    OnEnd< override>():void=
        # TODO: Replace this with your code
        Log("Goodbye, NPC!",LOG_LEVEL_NPCDEBUG)


    NPCDamaged(DamageResult : damage_result) : void =
        Character:=DamageResult.Target
        if(Prop := creative_prop[Character]):
            #Prop.SetMaterial(Mat_Green_FollowMe)
            #Prop.SetMaterial(Mat_Green_FollowMe, ?Index:=0)
        #Print("NPC DAMAGED!")

