# fill this file with your verse-script
using { /UnrealEngine.com/Temporary/UI }
using { /UnrealEngine.com/Temporary/SpatialMath }
using { /Fortnite.com/UI }
using { /Verse.org/Simulation }
using { /Verse.org/Colors/NamedColors }
using { /Fortnite.com/Characters }

# A Verse-authored creative device that can be placed in a level
options_menu_ui := class():

    BakeryMachinesManager : bakery_machines_manager

    #Class settings
    var MaybeAgent : ?agent = false
    var MaybeCanvas : ?canvas = false
    var UIIsDisplayed :logic = false

    var ConfirmationToAll : int = 0

    ShowQuestTracker_Button : button_regular = button_regular{}
    SwitchMusic_Button : button_regular = button_regular{}
    ChangeLanguage_Button : button_regular = button_regular{}
    Cancel_Button : button_loud = button_loud{}

    var TimeElapsed : int = 0

    InitUI() : void =

        Canvas := MakeOptionsMenuCanvas()
        set MaybeCanvas = option{Canvas}

        ShowQuestTracker_Button.SetText(BasicMessage("Show quest tracker"))
        if(BakeryMachinesManager.PlayerDataCache.Language = ENGLISH):
            ChangeLanguage_Button.SetText(BasicMessage("Change Language"))
        
        if(BakeryMachinesManager.PlayerDataCache.Language = SPANISH):
            ChangeLanguage_Button.SetText(BasicMessage("Cambiar Idioma"))
        
        if(BakeryMachinesManager.PlayerDataCache.Language = FRENCH):
            ChangeLanguage_Button.SetText(BasicMessage("Changer la Langue"))
        
        if(BakeryMachinesManager.PlayerDataCache.Language = PORTUGUESE):
            ChangeLanguage_Button.SetText(BasicMessage("Mudar Idioma"))
        
        SwitchMusic_Button.SetText(BasicMessage("Turn off music"))
        Cancel_Button.SetText(BasicMessage("Close "))

        Subscribers()

    Subscribers() : void =
        ShowQuestTracker_Button.OnClick().Subscribe(ShowQuestTracker) 
        SwitchMusic_Button.OnClick().Subscribe(SwitchMusic) 
        Cancel_Button.OnClick().Subscribe(Remove_Canvas) 
        ChangeLanguage_Button.OnClick().Subscribe(ChangeLanguage) 

    ShowQuestTracker(Message : widget_message) : void =
        if(Agent:=MaybeAgent?, Player := player[Agent], PlayerData:=BakeryMachinesManager.PlayerDataCache):
            var NewPlayerData : player_data_table = PlayerData
            if(BakeryMachinesManager.PlayerDataCache.IsQuestTrackerOn?):
                set NewPlayerData = BakeryMachinesManager.PlayersDataManager.SetTrackerState(PlayerData,false)
                BakeryMachinesManager.SetQuestTrackerState(false)
                #BakeryMachinesManager.CheckMusicPlay()
            else:
                set NewPlayerData = BakeryMachinesManager.PlayersDataManager.SetTrackerState(PlayerData,true)
                BakeryMachinesManager.SetQuestTrackerState(true)
                #BakeryMachinesManager.CheckMusicPlay()
            #BakeryMachinesManager.SetNewPlayerDataCache(PlayerData)
            #BakeryMachinesManager.PlayersDataManager.SaveNewData(Player,PlayerData,NewPlayerData)
        UpdateUI()
        Remove_Canvas(Message)


    ChangeLanguage(Message : widget_message) : void =
        if(Agent:=MaybeAgent?, Player := player[Agent], PlayerData:=BakeryMachinesManager.PlayerDataCache):
        UpdateUI()
        Remove_Canvas(Message)
        spawn:
            DelayChangeLanguage()

    DelayChangeLanguage()<suspends> : void =
        Sleep(0.1)
        BakeryMachinesManager.Display_LanguageMenu()
        
    SwitchMusic(Message : widget_message) : void =
        if(Agent:=MaybeAgent?, Player := player[Agent], Character := Player.GetFortCharacter[]):
            if(BakeryMachinesManager.PlayerDataCache.IsMusicOn?):
                #set NewPlayerData = BakeryMachinesManager.PlayersDataManager.SetMusicState(PlayerData,false)
                BakeryMachinesManager.SetMusicState(false)
            else:
                #set NewPlayerData = BakeryMachinesManager.PlayersDataManager.SetMusicState(PlayerData,true)
                BakeryMachinesManager.SetMusicState(true)
            var MinDistanceFromHardBread : float = 9999999.9
            var HardBreadLevel : int = -1

            #Check if agent is Playing a Dungeon
            NumDungeon := BakeryMachinesManager.GameManager.GeneralMobsManager.IsAgentInsideDungeon(Agent)
            if(NumDungeon<>-1):
                if(IsFinalMusicIsPlaying:=BakeryMachinesManager.GameManager.GeneralMobsManager.AllDungeons_FinalMusicIsPlaying[NumDungeon]):
                    if(IsFinalMusicIsPlaying?):
                        BakeryMachinesManager.CheckMusicPlay(DUNGEONMINIBOSSMUSIC_SOUND)
                    else:
                        BakeryMachinesManager.CheckMusicPlay(DUNGEONMUSIC_SOUND)
            else:
                for(HB : BakeryMachinesManager.GameManager.AllHardBreads, DistanceBetween:=Distance(HB.SlateBreadProp.GetTransform().Translation,Character.GetTransform().Translation), MinDistanceFromHardBread>DistanceBetween):
                    set HardBreadLevel = HB.Level
                    set MinDistanceFromHardBread = DistanceBetween
                if(HardBreadLevel>13):
                    BakeryMachinesManager.CheckMusicPlay(SECRETMUSIC_SOUND)
                else if(HardBreadLevel>7):
                    BakeryMachinesManager.CheckMusicPlay(DESERTMUSIC_SOUND)
                else:
                    BakeryMachinesManager.CheckMusicPlay(BAKERSMUSIC_SOUND)
            #BakeryMachinesManager.SetNewPlayerDataCache(PlayerData)
            #BakeryMachinesManager.PlayersDataManager.SaveNewData(Player,PlayerData,NewPlayerData)
        UpdateUI()
            
    SetAgent(Agent : agent) : void =
        set MaybeAgent = option{Agent}
        set UIIsDisplayed = false
        UpdateUI()

    UpdateUI() : void =
        if(BakeryMachinesManager.PlayerDataCache.IsMusicOn?):
            SwitchMusic_Button.SetText(BasicMessage("Turn off music"))
        else:
            SwitchMusic_Button.SetText(BasicMessage("Turn on music"))
        if(BakeryMachinesManager.PlayerDataCache.IsQuestTrackerOn?):
            ShowQuestTracker_Button.SetText(BasicMessage("Hide quest tracker"))
        else:
            ShowQuestTracker_Button.SetText(BasicMessage("Show quest tracker"))

    Remove_Canvas(Message : widget_message) : void =
        set UIIsDisplayed = false
        set TimeElapsed = 5
        if:
            Agent:=MaybeAgent?
            Player:=player[Agent]
            PlayerUI:=GetPlayerUI[Player]
            Canvas := MaybeCanvas?
        then:
            PlayerUI.RemoveWidget(Canvas)
            RemoveDelay_OnRemove()

    Remove_CanvasWithoutMessage() : void =
        set UIIsDisplayed = false
        set TimeElapsed = 5
        if:
            Agent:=MaybeAgent?
            Player:=player[Agent]
            PlayerUI:=GetPlayerUI[Player]
            Canvas := MaybeCanvas?
        then:
            PlayerUI.RemoveWidget(Canvas)
            RemoveDelay_OnRemove()

    RemoveDelay_OnAdd() : void =
        Cancel_Button.SetText(BasicMessage("CLOSE"))


    RemoveDelay_OnRemove() : void =
        Cancel_Button.SetText(BasicMessage("CLOSE "))

    StartTimeElapsed()<suspends> : void =
        loop:
            Log("TimeElapsed: {TimeElapsed}", LOG_LEVEL_DEBUG9)
            if(TimeElapsed > 5):
                ##Print("Closing the UI because 20 seconds has pased")
                Remove_CanvasWithoutMessage()
                break
            if(UIIsDisplayed = false):
                ##Print("Breaking Time Elapsed loop")
                break
            Sleep(1.0)
            set TimeElapsed = TimeElapsed + 1
            
    Display_Canvas() : void =
        if(not UIIsDisplayed?):
            set UIIsDisplayed = true
            set TimeElapsed = 0
            spawn: 
                StartTimeElapsed()
            set ConfirmationToAll = 0
            set UIIsDisplayed = true
            var Index : int = 0
            if:
                Agent:=MaybeAgent?
                PlayerUI:=GetPlayerUI[player[Agent]]
                Canvas := MaybeCanvas?
            then:
                PlayerUI.AddWidget(Canvas,player_ui_slot{ZOrder:=1, InputMode:=ui_input_mode.All})
                RemoveDelay_OnAdd()   
            UpdateUI()

    MakeOptionsMenuCanvas(): canvas=
        Canvas := canvas:
            Slots:=array:
                #Background image
                canvas_slot:
                    Anchors:=anchors{Minimum:=vector2{X:=0.5,Y:=0.5},Maximum:=vector2{X:=0.5,Y:=0.5}}
                    ZOrder:=0
                    Offsets:=margin{Left:=0.0,Top:=0.0,Right:=0.0,Bottom:=0.0} #PositionX, PositionY, SizeX, SizeY
                    Alignment:=vector2{X:=0.5,Y:=0.5}
                    Widget:=overlay:
                        Slots:=array:
                            
                            overlay_slot:
                                HorizontalAlignment:=horizontal_alignment.Fill
                                VerticalAlignment:=vertical_alignment.Fill
                                Widget:=texture_block{DefaultImage:=Images.BlackBackground}
                            overlay_slot:
                                HorizontalAlignment:=horizontal_alignment.Fill
                                VerticalAlignment:=vertical_alignment.Fill
                                Padding:=margin{Left:=SelectorBoxPadding,Right:=SelectorBoxPadding,Top:=SelectorBoxPadding,Bottom:=SelectorBoxPadding}
                                Widget:=texture_block{DefaultImage:=Images.BlackBackground}
                            overlay_slot: #Screw 0
                                Padding:=margin{Left:=3.5,Top:=3.5,Right:=3.5,Bottom:=3.5}
                                HorizontalAlignment:=horizontal_alignment.Left
                                VerticalAlignment:=vertical_alignment.Top
                                Widget:=texture_block{DefaultImage:=Images.Dialogue_Bolt, DefaultDesiredSize:=vector2{X:=5.0,Y:=5.0}}
                            overlay_slot: #Screw 1
                                Padding:=margin{Left:=3.5,Top:=3.5,Right:=3.5,Bottom:=3.5}
                                HorizontalAlignment:=horizontal_alignment.Right
                                VerticalAlignment:=vertical_alignment.Top
                                Widget:=texture_block{DefaultImage:=Images.Dialogue_Bolt, DefaultDesiredSize:=vector2{X:=5.0,Y:=5.0}}
                            overlay_slot: #Screw 2
                                Padding:=margin{Left:=3.5,Top:=3.5,Right:=3.5,Bottom:=3.5}
                                HorizontalAlignment:=horizontal_alignment.Left
                                VerticalAlignment:=vertical_alignment.Bottom
                                Widget:=texture_block{DefaultImage:=Images.Dialogue_Bolt, DefaultDesiredSize:=vector2{X:=5.0,Y:=5.0}}
                            overlay_slot: #Screw 3
                                Padding:=margin{Left:=3.5,Top:=3.5,Right:=3.5,Bottom:=3.5}
                                HorizontalAlignment:=horizontal_alignment.Right
                                VerticalAlignment:=vertical_alignment.Bottom
                                Widget:=texture_block{DefaultImage:=Images.Dialogue_Bolt, DefaultDesiredSize:=vector2{X:=5.0,Y:=5.0}}
                                <#
                            overlay_slot:
                                HorizontalAlignment:=horizontal_alignment.Fill
                                VerticalAlignment:=vertical_alignment.Fill
                                Widget:=texture_block{DefaultImage:=Images.BlackBackground,DefaultDesiredSize:=vector2{X:=5.0,Y:=5.0}}#>
                            overlay_slot:
                                Padding:=margin{Top:=10.0,Bottom:=10.0,Right:=10.0,Left:=10.0}
                                Widget:=stack_box:
                                    Orientation:=orientation.Vertical
                                        Slots:=array:
                                            stack_box_slot: 
                                                Widget:=overlay:
                                                    Slots:=array:
                                                        overlay_slot:
                                                            Padding:=margin{Top:=5.0,Bottom:=5.0,Right:=5.0,Left:=5.0}
                                                            Widget:=text_block{DefaultText:=BasicMessage("OPTIONS MENU"), DefaultTextColor:=White, DefaultShadowOffset:=ShadowOffset, DefaultShadowOpacity := DefaultTextShadowOpacity}
                                            stack_box_slot: 
                                                Padding:=margin{Top:=5.0,Bottom:=5.0,Right:=5.0,Left:=5.0}
                                                Widget:=overlay:
                                                    Slots:=array:
                                                        overlay_slot:
                                                            Widget:=texture_block{DefaultImage:=Images.EmptyTexture, DefaultDesiredSize:=vector2{X:=500.0,Y:=80.0}}
                                                        overlay_slot:
                                                            HorizontalAlignment:=horizontal_alignment.Fill
                                                            VerticalAlignment:=vertical_alignment.Fill
                                                            Padding:=margin{Top:=15.0,Bottom:=5.0,Right:=5.0,Left:=5.0}
                                                            Widget:=ShowQuestTracker_Button
                                            stack_box_slot: 
                                                Padding:=margin{Top:=5.0,Bottom:=5.0,Right:=5.0,Left:=5.0}
                                                Widget:=overlay:
                                                    Slots:=array:
                                                        overlay_slot:
                                                            Widget:=texture_block{DefaultImage:=Images.EmptyTexture, DefaultDesiredSize:=vector2{X:=500.0,Y:=80.0}}
                                                        overlay_slot:
                                                            HorizontalAlignment:=horizontal_alignment.Fill
                                                            VerticalAlignment:=vertical_alignment.Fill
                                                            Padding:=margin{Top:=15.0,Bottom:=15.0,Right:=5.0,Left:=5.0}
                                                            Widget:=SwitchMusic_Button
                                            stack_box_slot: 
                                                Padding:=margin{Top:=5.0,Bottom:=5.0,Right:=5.0,Left:=5.0}
                                                Widget:=overlay:
                                                    Slots:=array:
                                                        overlay_slot:
                                                            Widget:=texture_block{DefaultImage:=Images.EmptyTexture, DefaultDesiredSize:=vector2{X:=500.0,Y:=80.0}}
                                                        overlay_slot:
                                                            HorizontalAlignment:=horizontal_alignment.Fill
                                                            VerticalAlignment:=vertical_alignment.Fill
                                                            Padding:=margin{Top:=15.0,Bottom:=15.0,Right:=5.0,Left:=5.0}
                                                            Widget:=ChangeLanguage_Button
                                            stack_box_slot: 
                                                Padding:=margin{Top:=15.0,Bottom:=15.0,Right:=5.0,Left:=5.0}
                                                Widget:=overlay:
                                                    Slots:=array:
                                                        overlay_slot:
                                                            Widget:=texture_block{DefaultImage:=Images.EmptyTexture, DefaultDesiredSize:=vector2{X:=500.0,Y:=80.0}}
                                                        overlay_slot:
                                                            HorizontalAlignment:=horizontal_alignment.Fill
                                                            VerticalAlignment:=vertical_alignment.Fill
                                                            Padding:=margin{Top:=5.0,Bottom:=5.0,Right:=5.0,Left:=5.0}
                                                            Widget:=Cancel_Button

MakeOptionsMenuUI<constructor>(InBakeryMachinesManager : bakery_machines_manager)<transacts> := options_menu_ui :
    BakeryMachinesManager := InBakeryMachinesManager