using { /Verse.org/Simulation }
using { /Verse.org/Random }


MainFormula_PerkPetCost(PetStatType : int, CurrentLV : int) : float =
    var Cost : float = 10.0
    if(PetStatType=STAT_TYPE_PET_PERK_MAXACTIVE):
        if(CurrentLV=1):
            set Cost = 0.0
        else if(CurrentLV=2):
            set Cost = 5.0
        else if(CurrentLV=3):
            set Cost = 20.0
        #else:
        #    set Cost = ActivePetSlotsCostBase*1.0*CurrentLV
    else if(PetStatType=STAT_TYPE_PET_PERK_MAXSLOTS):
        return Pow(PetSlotsCostBase,CurrentLV*1.0)
    else if(PetStatType=STAT_TYPE_PET_PERK_TIER):  
        if(CurrentLV=1):
            set Cost = 0.0
        else if(CurrentLV=2):
            set Cost = 10.0
        else if(CurrentLV=3):
            set Cost = 40.0


        <#
        if(CurrentLV=2):
            set Cost = PetTierCostBase-3.0
        else:
            set Cost = PetTierCostBase*1.0*CurrentLV#>
    return Cost


MainFormula_EnhancementsPetStats(PetStatType : int, PetType : int, StatLV : int, Purity : int) : float =
    var FinalAmount : float = 0.0
    var MulPurity : float = 1.0

    if(Value:=Floor[Purity*1.0/29.0], set MulPurity = Value*1.0):
    if(MulPurity<=0.0):
        set MulPurity = 0.5

    #ApplyPurity
    Multiplier:=StatLV*MulPurity
    if(PetStatType = STAT_TYPE_PLAYER_PICKDAMAGE):
        set FinalAmount = 1.0+(ALL_PLAYER_PICKDAMAGE_MULTIPLIER * Multiplier)
        if(FinalAmount<1.01):
            set FinalAmount = 1.01
    else if(PetStatType = STAT_TYPE_PLAYER_WEAKPOINTMULT):
        set FinalAmount = 1.0+(ALL_PLAYER_WEAKPOINT_MULTIPLIER * Multiplier)
        if(FinalAmount<1.01):
            set FinalAmount = 1.01
    else if(PetStatType = STAT_TYPE_PLAYER_PICKEFFICENCY):
        set FinalAmount = 1.0+(ALL_PLAYER_PICKEFFICIENCY_MULTIPLIER * Multiplier)
        if(FinalAmount<1.01):
            set FinalAmount = 1.01
    else if(PetStatType = STAT_TYPE_PLAYER_PICKCRITICALMULT):
        set FinalAmount = 1.0+(ALL_PLAYER_PICKCRITICALDAMAGE_MULTIPLIER * Multiplier)
        if(FinalAmount<1.01):
            set FinalAmount = 1.01
    else if(PetStatType = STAT_TYPE_PLAYER_SPECIAL_A): #Specials A
        if(PetType = ADC):
            set FinalAmount = 1.0+(ADC_PLAYER_SPECIALA_BOWMULTIPLIER * Multiplier)
        else if(PetType = DRUID):
            set FinalAmount = DRUID_PLAYER_SPECIALA_HEAL * Multiplier
        else if(PetType = TANK):
            set FinalAmount = TANK_PLAYER_SPECIALA_HEALTH * Multiplier
        else if(PetType = FIGHTER):
            set FinalAmount = 1.0+(FIGHTER_PLAYER_SPECIALA_SWORDDAMAGE * Multiplier)
        else if(PetType = HUNTER): 
            set FinalAmount = 1.0+(HUNTER_PLAYER_SPECIALA_HUNTPROB * Multiplier)
    else if(PetStatType = STAT_TYPE_PLAYER_SPECIAL_B): #Specials B
        if(PetType = ADC):
            set FinalAmount = ADC_PLAYER_SPECIALB_BASETIME-(ADC_PLAYER_SPECIALB_IMPROVETIME * Multiplier)
        else if(PetType = DRUID):
            set FinalAmount = DRUID_PLAYER_SPECIALB_EMERGENCY * Multiplier
        else if(PetType = TANK):
            set FinalAmount = TANK_PLAYER_SPECIALB_SHIELD * Multiplier
        else if(PetType = FIGHTER):
            set FinalAmount = 1.0+(FIGHTER_PLAYER_SPECIALB_ULTRASTRIKE * Multiplier)
        else if(PetType = HUNTER): 
            set FinalAmount = 1.0+(HUNTER_PLAYER_SPECIALB_WEAPONDAMAGE * Multiplier)
    return FinalAmount

MainFormula_TournamentPetStats(PetStatType : int, PetType : int, StatLV : int, Purity : int) : float =
    var FinalAmount : float = 0.0
    var MulPurity : float = 0.25

    MulTier := 1.0-MulPurity #Old Multiplier was 1.0, now is 0.25 so each tier is plus 0.75*100 = 75.0

    if(Purity=0):
        set MulPurity = 0.1

    if(Purity>0):
        set MulPurity = Purity*0.25 

        TuplePurity:=ConvertRawPurity(Purity)
        ConvertedTier:=TuplePurity(0)

        set MulPurity = MulPurity + (ConvertedTier-1)*(MulTier*100.0)

    

    #ApplyPurity
    PurityRate:=(MulPurity*1.0)/250.0
    Multiplier:=StatLV*PurityRate

    if(PetStatType = STAT_TYPE_RANGE):
        if(PetType = DRUID or PetType = ADC):
            if(SpecialRange:=Floor(Purity/100)):
                return (SpecialRange*1.0)+2.0
        return 1.0

    if(PetType = DRUID):
        if(PetStatType = STAT_TYPE_HEALTH):
            set FinalAmount = Multiplier*DRUID_HEALTH_MULTIPLIER*PetBase_Health
        else if(PetStatType = STAT_TYPE_DAMAGE):
            set FinalAmount = Multiplier*DRUID_DAMAGE_MULTIPLIER*PetBase_Damage
        else if(PetStatType = STAT_TYPE_HEALING):
            set FinalAmount = Multiplier*DRUID_HEALING_MULTIPLIER*PetBase_HealingPerk
        else if(PetStatType = STAT_TYPE_ATTACKSPEED):
            Ammount : float = PetBase_AttackSpeed - (PetBase_AttackSpeedImprovement * DRUID_ATTACKSPEED_MULTIPLIER * StatLV)
            if(Ammount<0.1):
                return 0.1
            else:
                return Ammount
        else if(PetStatType = STAT_TYPE_SPEED):
            Ammount : float = PetBase_Speed - (PetBase_SpeedImprovement * DRUID_SPEED_MULTIPLIER * StatLV)
            if(Ammount<0.1):
                return 0.1
            else:
                return Ammount
    else if(PetType = ADC):
        if(PetStatType = STAT_TYPE_HEALTH):
            set FinalAmount = Multiplier*ADC_HEALTH_MULTIPLIER*PetBase_Health
        else if(PetStatType = STAT_TYPE_DAMAGE):
            set FinalAmount = Multiplier*ADC_DAMAGE_MULTIPLIER*PetBase_Damage
        else if(PetStatType = STAT_TYPE_HEALING):
            set FinalAmount = Multiplier*ADC_HEALING_MULTIPLIER*PetBase_HealingPerk
        else if(PetStatType = STAT_TYPE_ATTACKSPEED):
            Ammount : float = PetBase_AttackSpeed - (PetBase_AttackSpeedImprovement * ADC_ATTACKSPEED_MULTIPLIER * StatLV)
            if(Ammount<0.1):
                return 0.1
            else:
                return Ammount
        else if(PetStatType = STAT_TYPE_SPEED):
            Ammount : float = PetBase_Speed - (PetBase_SpeedImprovement * ADC_SPEED_MULTIPLIER * StatLV)
            if(Ammount<0.1):
                return 0.1
            else:
                return Ammount
    else if(PetType = FIGHTER):
        if(PetStatType = STAT_TYPE_HEALTH):
            set FinalAmount = Multiplier*FIGHTER_HEALTH_MULTIPLIER*PetBase_Health
        else if(PetStatType = STAT_TYPE_DAMAGE):
            set FinalAmount = Multiplier*FIGHTER_DAMAGE_MULTIPLIER*PetBase_Damage
        else if(PetStatType = STAT_TYPE_HEALING):
            set FinalAmount = StatLV*FIGHTER_HEALING_MULTIPLIER*PetBase_HealingPerk
        else if(PetStatType = STAT_TYPE_ATTACKSPEED):
            Ammount : float = PetBase_AttackSpeed - (PetBase_AttackSpeedImprovement * FIGHTER_ATTACKSPEED_MULTIPLIER * StatLV)
            if(Ammount<0.1):
                return 0.1
            else:
                return Ammount
        else if(PetStatType = STAT_TYPE_SPEED):
            Ammount : float = PetBase_Speed - (PetBase_SpeedImprovement * FIGHTER_SPEED_MULTIPLIER * StatLV)
            if(Ammount<0.1):
                return 0.1
            else:
                return Ammount
    else if(PetType = TANK):
        if(PetStatType = STAT_TYPE_HEALTH):
            set FinalAmount = Multiplier*TANK_HEALTH_MULTIPLIER*PetBase_Health
        else if(PetStatType = STAT_TYPE_DAMAGE):
            set FinalAmount = Multiplier*TANK_DAMAGE_MULTIPLIER*PetBase_Damage
        else if(PetStatType = STAT_TYPE_HEALING):
            set FinalAmount = Multiplier*TANK_HEALING_MULTIPLIER*PetBase_HealingPerk
        else if(PetStatType = STAT_TYPE_ATTACKSPEED):
            Ammount : float = PetBase_AttackSpeed - (PetBase_AttackSpeedImprovement * TANK_ATTACKSPEED_MULTIPLIER * StatLV)
            if(Ammount<0.1):
                return 0.1
            else:
                return Ammount
        else if(PetStatType = STAT_TYPE_SPEED):
            Ammount : float = PetBase_Speed - (PetBase_SpeedImprovement * TANK_SPEED_MULTIPLIER * StatLV)
            if(Ammount<0.1):
                return 0.1
            else:
                return Ammount
    else if(PetType = HUNTER): 
        if(PetStatType = STAT_TYPE_HEALTH):
            set FinalAmount = Multiplier*HUNTER_HEALTH_MULTIPLIER*PetBase_Health
        else if(PetStatType = STAT_TYPE_DAMAGE):
            set FinalAmount = Multiplier*HUNTER_DAMAGE_MULTIPLIER*PetBase_Damage
        else if(PetStatType = STAT_TYPE_HEALING):
            set FinalAmount = Multiplier*HUNTER_HEALING_MULTIPLIER*PetBase_HealingPerk
        else if(PetStatType = STAT_TYPE_ATTACKSPEED):
            Ammount : float = PetBase_AttackSpeed - (PetBase_AttackSpeedImprovement * HUNTER_ATTACKSPEED_MULTIPLIER * StatLV)
            if(Ammount<0.1):
                return 0.1
            else:
                return Ammount
        else if(PetStatType = STAT_TYPE_SPEED):
            Ammount : float = PetBase_Speed - (PetBase_SpeedImprovement * HUNTER_SPEED_MULTIPLIER * StatLV)
            if(Ammount<0.1):
                return 0.1
            else:
                return Ammount
    return FinalAmount

MainFormula_PackageACost(LVPackageA : int, ?AmountCoal : float = 0.0, ?Inverted : logic = false) : float =
    if(not Inverted?):
        return LVPackageA*PetPackageA_BaseCost
    else:
        return AmountCoal/PetPackageA_BaseCost

MainFormula_PackageBCost(LVPackageB : int, ?AmountCoal : float = 0.0, ?Inverted : logic = false) : float =
    if(not Inverted?):
        return LVPackageB*PetPackageB_BaseCost
    else:
        return AmountCoal/PetPackageB_BaseCost


MainFormula_PetPurityCost(Purity : int) : float = 
    <#
    MaxMutantAmount:=Pow(2.0,HB_MAXLEVEL*1.0)
    
    TotalMaxMutantAmount := MaxMutantAmount + ((MaxMutantAmount/100.0)*NumHitsAccumulated)
    ReqMutantMaterial_ForMaxPet := TotalMaxMutantAmount*25#>
    var Value : float = 0.0

    NumHitsAccumulated:=(MinNum_AVGHit+MaxNum_AVGHit)/2.0
    NumPurity_PerHBLevel:=(PetTotalMaxPurity*1.0)/(HB_MAXLEVEL*1.0)
    FloatHB_LV:=(Purity*1.0)/NumPurity_PerHBLevel

    if(HB_LV:=Floor[FloatHB_LV]):
        DecimalDifference := FloatHB_LV - (HB_LV*1.0)

        Req_MutantMaterial:=CheckFinalMutantMaterialAward(HB_LV,NumHitsAccumulated,1.0)*ReqDestroyedHB_PerPurity
        NextLevel_Req_MutantMaterial:=CheckFinalMutantMaterialAward(HB_LV+1,NumHitsAccumulated,1.0)*ReqDestroyedHB_PerPurity

        MutantDifference:=NextLevel_Req_MutantMaterial-Req_MutantMaterial

        set Value = Req_MutantMaterial + (DecimalDifference*MutantDifference)

    if(Value<1000.0, CleanedValue:=Ceil[Value]):
        set Value = CleanedValue*1.0

    return Value


    #OLD CODE
    <#
    if(not Inverted?):
        return Pow((PetBasePurityCost*Purity*2.0),2.0)  
    else: 
        return ((Sqrt(AmountMutantMaterial))/PetBasePurityCost)/2.0#>

ConvertRange(TilesToCmMode : logic, Amount : float) : float =
    if(TilesToCmMode?):
        if(Amount>=UltraRangeTilesRange):
            return UltraRangeAttackRange
        else if(Amount>=SuperRangeTilesRange):
            return SuperRangeAttackRange
        else if(Amount>=LongRangeTilesRange):
            return LongRangeAttackRange
        else:
            return MeleeAttackRange
    else:
        if(Amount>=UltraRangeAttackRange):
            return UltraRangeTilesRange
        else if(Amount>=SuperRangeAttackRange):
            return SuperRangeTilesRange
        else if(Amount>=LongRangeAttackRange):
            return LongRangeTilesRange
        else:
            return MeleeTilesRange


ConvertRawPurity(TotalPurity : int) : tuple(int,int) =
    var ConvertedPurity : int = 0
    var ConvertedTier : int = 0
    if:
        NumTiers:=Floor[(TotalPurity*1.0)/100.0]+1
        NumPurity:=Floor[(TotalPurity*1.0)-(100.0*(NumTiers-1))]
    then:
        set ConvertedTier = NumTiers
        set ConvertedPurity = NumPurity
    return (ConvertedTier,ConvertedPurity)

#Calcule Pet Stat
CalculePetStat<public>(StatToCalcule : int, CurrentStatLV : int, PetType : int, Purity : int) : tuple(float,float,float) =
    #### Next stats
    var ObjetiveStatLV : int = CurrentStatLV+1
    var NextStat : float = 0.0
    var CurrentStat : float = 0.0

    var NextCost : float = 0.0
    var CurrentCost : float = 0.0

    if(StatToCalcule = STAT_TYPE_PLAYER_SPECIAL_A or StatToCalcule = STAT_TYPE_PLAYER_PICKDAMAGE or StatToCalcule = STAT_TYPE_PLAYER_WEAKPOINTMULT or StatToCalcule = STAT_TYPE_PLAYER_SPECIAL_B or StatToCalcule = STAT_TYPE_PLAYER_PICKEFFICENCY or StatToCalcule = STAT_TYPE_PLAYER_PICKCRITICALMULT):
        set NextStat = MainFormula_EnhancementsPetStats(StatToCalcule, PetType, ObjetiveStatLV, Purity)
        set CurrentStat = MainFormula_EnhancementsPetStats(StatToCalcule, PetType, CurrentStatLV, Purity)
    else:
        set NextStat = MainFormula_TournamentPetStats(StatToCalcule, PetType, ObjetiveStatLV, Purity)
        set CurrentStat = MainFormula_TournamentPetStats(StatToCalcule, PetType, CurrentStatLV, Purity)

    #### Costs
    if(StatToCalcule=STAT_TYPE_HEALTH or StatToCalcule=STAT_TYPE_DAMAGE or StatToCalcule=STAT_TYPE_SPEED or StatToCalcule= STAT_TYPE_PLAYER_SPECIAL_A or StatToCalcule=STAT_TYPE_PLAYER_PICKDAMAGE or StatToCalcule=STAT_TYPE_PLAYER_WEAKPOINTMULT):
        set NextCost = MainFormula_PackageACost(ObjetiveStatLV)
        set CurrentCost = MainFormula_PackageACost(CurrentStatLV)
    else: 
        set NextCost = MainFormula_PackageBCost(ObjetiveStatLV)
        set CurrentCost = MainFormula_PackageBCost(CurrentStatLV)
    #### Difference
    TotalDifference := NextStat-CurrentStat
    return(CurrentStat,NextCost,TotalDifference)

CalculePetCost(Purity : int) : float =
    return MainFormula_PetPurityCost(Purity)
        
CalculePetTotalAmountsInvested(Pet : pet_data_table)<suspends> : tuple(float,float,float) =
    var TotalMutantMaterialAmount : float = 0.0
    var TotalCoalAmountForPackageA : float = 0.0
    var TotalCoalAmountForPackageB : float = 0.0

    set TotalMutantMaterialAmount = TotalMutantMaterialAmount + MainFormula_PetPurityCost(Pet.Purity)
    for(LV:=1..Pet.LVPet_PackageA):
        if(LV<>1):
            set TotalCoalAmountForPackageA = TotalCoalAmountForPackageA + MainFormula_PackageACost(LV)
    Sleep(0.04)
    for(LV:=1..Pet.LVPet_PackageB):
        if(LV<>1):
            set TotalCoalAmountForPackageB = TotalCoalAmountForPackageB + MainFormula_PackageBCost(LV)

    return (TotalMutantMaterialAmount,TotalCoalAmountForPackageA,TotalCoalAmountForPackageB)

CalculeNumChallengesPerCollectible(AllTotalNumChallenges : int, AllCollectibleHexaChallenges : int) : int =
    var NumChallenges : int = 22
    if(Amount:=Ceil[(AllTotalNumChallenges*1.0)/(AllCollectibleHexaChallenges*1.0)]):

        #ValidBakery.HexaGameManager.HexaChallengesDatabase.AllChallenges.Length
        set NumChallenges = Amount 
    return NumChallenges

CalculeNumChallengesUnlocked(AllHexaGame_TotemsUnlocked : []logic, AllTotalNumChallenges : int, AllCollectibleHexaChallenges : int) : int =
    var NumTotems : int = 0
    for(TotemsUnlocked : AllHexaGame_TotemsUnlocked, TotemsUnlocked?):
        set NumTotems = NumTotems + 1

    var NumChallengesPerTotem : int = 22
    if(Amount:=Ceil[(AllTotalNumChallenges*1.0)/(AllCollectibleHexaChallenges*1.0)]):

        #ValidBakery.HexaGameManager.HexaChallengesDatabase.AllChallenges.Length
        set NumChallengesPerTotem = Amount 
    return NumTotems * NumChallengesPerTotem



    

    
                            
      

    

    


