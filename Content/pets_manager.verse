
using { /Fortnite.com/Devices }
using { /Verse.org/Simulation }
using { /UnrealEngine.com/Temporary/Diagnostics }
using { /UnrealEngine.com/Temporary/UI }
using { /UnrealEngine.com/Temporary/SpatialMath }
using { /Fortnite.com/UI }
using { /Verse.org/Colors/NamedColors }
using { /Fortnite.com/Characters }
using { /Verse.org/Random }


# See https://dev.epicgames.com/documentation/en-us/uefn/create-your-own-device-in-verse for how to create a verse device.

# A Verse-authored creative device that can be placed in a level
pets_manager := class():

    #Class settings
    var MaybeAgent : ?agent = false
    var GameManager : game_manager

    RowPadding:margin=margin{Left:=5.0,Top:=5.0,Right:=5.0,Bottom:=5.0}

    Empty_TextureBlock : texture_block = texture_block{DefaultImage:=Images.EmptyTexture}
    Empty_TextBlock : text_block = text_block{}

    #PLAYER ENHACEMENTS
    #######
        #COMMON
    var PlayerEnhacement_PickDamage : float = 1.0
    var PlayerEnhacement_WeakPointMult : float = 1.0
    var PlayerEnhacement_PickEfficency : float = 1.0
    var PlayerEnhacement_PickCriticalDam : float = 1.0
        #ADC
    var PlayerEnhacement_BowExtraDamage : float = 1.0
    var PlayerEnhacement_ArrowCraft : float = 0.0
        #DRUID
    var PlayerEnhacement_HealthRegeneration : float = 0.0
    var PlayerEnhacement_EmergencyHealing : float = 0.0
    var PlayerEnhacement_HealLastTimeUsed : float = 1.0
        #TANK
    var PlayerEnhacement_ExtraHealth : float = 0.0
    var PlayerEnhacement_EmergencyShield : float = 0.0
    var PlayerEnhacement_ShieldLastTimeUsed : float = 1.0
        #FIGHTER
    var PlayerEnhacement_SwordExtraDamage : float = 1.0
    var PlayerEnhacement_SwordSuperStrike : float = 1.0
    var PlayerEnhacement_SwordSuperStrikeNum : float = 0.0 
    var LastHammerStrike : float = 0.1
        #HUNTER
    var PlayerEnhacement_MobHunter : float = 1.0
    var PlayerEnhacement_WeaponsExtraCritical : float = 1.0

#BUYSLOT
#######
    var BuySlots_MaybeCanvas : ?canvas = false
    var BuySlots_UIIsDisplayed : logic = false

    BUColumn1SizeX : float = 80.0
    BUColumn1SizeY : float = 80.0
    BUColumn2SizeX : float = 425.0
    BUColumn2SizeY : float = 80.0
    BUColumn3SizeX : float = 150.0
    BUColumn3SizeY : float = 80.0
    BUColumn4SizeX : float = 200.0
    BUColumn4SizeY : float = 80.0
    BUColumn5SizeX : float = 200.0
    BUColumn5SizeY : float = 80.0
    BUColumn6SizeX : float = 225.0
    BUColumn6SizeY : float = 80.0

    BuySlot_TotalAggedSapphires_TextBlock : text_block = text_block{DefaultTextColor := AgedSapphiresTextColor, DefaultShadowOffset:=ShadowOffset, DefaultShadowOpacity := DefaultTextShadowOpacity}
    BuySlot_TotalMutantMaterial_TextBlock : text_block = text_block{DefaultTextColor := MutantMaterialTextColor, DefaultShadowOffset:=ShadowOffset, DefaultShadowOpacity := DefaultTextShadowOpacity}

    BuySlot_ActiveSlots_Current_TextBlock : text_block = text_block{DefaultTextColor := CurrentTextColor, DefaultShadowOffset:=ShadowOffset, DefaultShadowOpacity := DefaultTextShadowOpacity}
    BuySlot_MaxSlots_Current_TextBlock : text_block = text_block{DefaultTextColor := CurrentTextColor, DefaultShadowOffset:=ShadowOffset, DefaultShadowOpacity := DefaultTextShadowOpacity}
    BuySlot_TierUnlock_Current_TextBlock : text_block = text_block{DefaultTextColor := CurrentTextColor, DefaultShadowOffset:=ShadowOffset, DefaultShadowOpacity := DefaultTextShadowOpacity}

    BuySlot_ActiveSlots_Upgrade_TextBlock : text_block = text_block{DefaultTextColor := UpgradeTextColor, DefaultShadowOffset:=ShadowOffset, DefaultShadowOpacity := DefaultTextShadowOpacity}
    BuySlot_MaxSlots_Upgrade_TextBlock : text_block = text_block{DefaultTextColor := UpgradeTextColor, DefaultShadowOffset:=ShadowOffset, DefaultShadowOpacity := DefaultTextShadowOpacity}
    BuySlot_TierUnlock_Upgrade_TextBlock : text_block = text_block{DefaultTextColor := UpgradeTextColor, DefaultShadowOffset:=ShadowOffset, DefaultShadowOpacity := DefaultTextShadowOpacity}

    BuySlot_ActiveSlots_Cost_TextBlock : text_block = text_block{DefaultTextColor := AgedSapphiresTextColor, DefaultShadowOffset:=ShadowOffset, DefaultShadowOpacity := DefaultTextShadowOpacity}
    BuySlot_MaxSlots_Cost_TextBlock : text_block = text_block{DefaultTextColor := MutantMaterialTextColor, DefaultShadowOffset:=ShadowOffset, DefaultShadowOpacity := DefaultTextShadowOpacity}
    BuySlot_TierUnlock_Cost_TextBlock : text_block = text_block{DefaultTextColor := AgedSapphiresTextColor, DefaultShadowOffset:=ShadowOffset, DefaultShadowOpacity := DefaultTextShadowOpacity}

    BuySlot_ActiveSlots_Button : button_quiet = button_quiet{}
    BuySlot_MaxSlots_Button : button_quiet = button_quiet{}
    BuySlot_TierUnlock_Button : button_quiet = button_quiet{}

    BuySlot_Close_Button : button_loud = button_loud{}

    BuySlot_ActiveSlots_ResourceIcon : texture_block = texture_block{DefaultImage:=Images.AgedSaphire_Icon, DefaultDesiredSize:=vector2{X:=IconSmallSizeX,Y:=IconSmallSizeY}}
    BuySlot_MaxSlots_ResourceIcon : texture_block = texture_block{DefaultImage:=Images.Mutant_Material_Icon, DefaultDesiredSize:=vector2{X:=IconSmallSizeX,Y:=IconSmallSizeY}}
    BuySlot_TierUnlock_ResourceIcon : texture_block = texture_block{DefaultImage:=Images.AgedSaphire_Icon, DefaultDesiredSize:=vector2{X:=IconSmallSizeX,Y:=IconSmallSizeY}}

#PetActions
######
    var ArrowCraftCreation_Progress : float = 0.0

#Pet follower controller
#######
    var AllEquippedPetsIndexes : []int = array{-1,-1,-1}
    var AllEquippedPetTables : []pet_data_table = array{}
    var AllPetsFollowProp : []creative_prop = array{}
    var AllPetsFollowProp_Tier : []int = array{}
#MAIN MENU
########
    var MainMenu_MaybeCanvas : ?canvas = false
    var MainMenu_UIIsDisplayed : logic = false
    
    MainMenu_CreatePetButton : button_regular = button_regular{}
    MainMenu_EquipPetButton : button_regular = button_regular{}
    MainMenu_ImprovePetButton : button_regular = button_regular{}
    MainMenu_BuySlotsPetButton : button_regular = button_regular{}
    MainMenu_RecyclePetButton : button_regular = button_regular{}
    MainMenu_MixPetsButton : button_regular = button_regular{}

    MainMenu_CloseButton : button_loud = button_loud{}

#UPGRADE PET
#########
    var MaybePetsUpgradesUI : ?pets_upgrades_ui = false
    
#EQUIP PET
########
    var EquipPet_MaybeCanvas : ?canvas = false
    var EquipPet_UIIsDisplayed : logic = false

    var EquipToCreate : logic = false

    #All
    var AllEquipPet_SlotInfo : []text_block = array{}
    var AllEquipPet_ChangeSlot : []button_regular = array{}
    var AllEquipPet_Text_PackageA_LV : []text_block = array{}
    var AllEquipPet_Text_PackageB_LV : []text_block = array{}
        #TOURNAMENT
    var AllEquipPet_TextureBlock_PetImages : []texture_block = array{}
    var AllEquipPet_TextBlock_Purity : []text_block = array{}
    var AllEquipPet_TextBlock_Class : []text_block = array{}
    var AllEquipPet_TextBlock_PackA : []text_block = array{}
    var AllEquipPet_TextBlock_PackB : []text_block = array{}
        #Package A
    var AllEquipPet_TextBlock_Description_SpecialPerkA : []text_block = array{}
    var AllEquipPet_TextBlock_Current_SpecialPerkA : []text_block = array{}
    var AllEquipPet_TextBlock_Current_PlayerPickDamage : []text_block = array{}
    var AllEquipPet_TextBlock_Current_PlayerWeakPoint : []text_block = array{}
        #Package B
    var AllEquipPet_TextBlock_Description_SpecialPerkB : []text_block = array{}
    var AllEquipPet_TextBlock_Current_SpecialPerkB : []text_block = array{}
    var AllEquipPet_TextBlock_Current_PlayerPickEfficiency : []text_block= array{}
    var AllEquipPet_TextBlock_Current_PlayerPickCriticalDamage : []text_block = array{}

    #PET 0
    EquipPet_0_SlotInfo : text_block = text_block{DefaultTextColor:=White, DefaultShadowOffset:=ShadowOffset, DefaultShadowOpacity := DefaultTextShadowOpacity}
    EquipPet_0_ChangeSlotButton : button_regular = button_regular{}
    EquipPet_0_Text_PackageA_LV : text_block = text_block{DefaultTextColor := White, DefaultShadowOffset:=ShadowOffset, DefaultShadowOpacity := DefaultTextShadowOpacity}
    EquipPet_0_Text_PackageB_LV : text_block = text_block{DefaultTextColor := White, DefaultShadowOffset:=ShadowOffset, DefaultShadowOpacity := DefaultTextShadowOpacity} 
        #TOURNAMENT
    var EquipPet_0_TextureBlock_PetImages : texture_block = texture_block{DefaultImage:=Images.EmptyTexture,DefaultDesiredSize:=vector2{X:=75.0,Y:=75.0}}
    var EquipPet_0_TextBlock_Purity : text_block = text_block{DefaultTextColor := CurrentTextColor, DefaultShadowOffset:=ShadowOffset, DefaultShadowOpacity := DefaultTextShadowOpacity}
    var EquipPet_0_TextBlock_Class : text_block = text_block{DefaultTextColor := CurrentTextColor, DefaultShadowOffset:=ShadowOffset, DefaultShadowOpacity := DefaultTextShadowOpacity}
    var EquipPet_0_TextBlock_PackA : text_block = text_block{DefaultTextColor := CurrentTextColor, DefaultShadowOffset:=ShadowOffset, DefaultShadowOpacity := DefaultTextShadowOpacity}
    var EquipPet_0_TextBlock_PackB : text_block = text_block{DefaultTextColor := CurrentTextColor, DefaultShadowOffset:=ShadowOffset, DefaultShadowOpacity := DefaultTextShadowOpacity}
        #Package A
    var EquipPet_0_TextBlock_Description_SpecialPerkA : text_block = text_block{DefaultTextColor := Pink, DefaultShadowOffset:=ShadowOffset, DefaultShadowOpacity := DefaultTextShadowOpacity}
    var EquipPet_0_TextBlock_Current_SpecialPerkA : text_block = text_block{DefaultTextColor := CurrentTextColor, DefaultShadowOffset:=ShadowOffset, DefaultShadowOpacity := DefaultTextShadowOpacity}
    var EquipPet_0_TextBlock_Current_PlayerPickDamage : text_block = text_block{DefaultTextColor := CurrentTextColor, DefaultShadowOffset:=ShadowOffset, DefaultShadowOpacity := DefaultTextShadowOpacity}
    var EquipPet_0_TextBlock_Current_PlayerWeakPoint : text_block = text_block{DefaultTextColor := CurrentTextColor, DefaultShadowOffset:=ShadowOffset, DefaultShadowOpacity := DefaultTextShadowOpacity}
        #Package B
    var EquipPet_0_TextBlock_Description_SpecialPerkB : text_block = text_block{DefaultTextColor := Pink, DefaultShadowOffset:=ShadowOffset, DefaultShadowOpacity := DefaultTextShadowOpacity}
    var EquipPet_0_TextBlock_Current_SpecialPerkB : text_block = text_block{DefaultTextColor := CurrentTextColor, DefaultShadowOffset:=ShadowOffset, DefaultShadowOpacity := DefaultTextShadowOpacity}
    var EquipPet_0_TextBlock_Current_PlayerPickEfficiency : text_block = text_block{DefaultTextColor := CurrentTextColor, DefaultShadowOffset:=ShadowOffset, DefaultShadowOpacity := DefaultTextShadowOpacity}
    var EquipPet_0_TextBlock_Current_PlayerPickCriticalDamage : text_block = text_block{DefaultTextColor := CurrentTextColor, DefaultShadowOffset:=ShadowOffset, DefaultShadowOpacity := DefaultTextShadowOpacity}

    #PET 1
    EquipPet_1_SlotInfo : text_block = text_block{DefaultTextColor:=White, DefaultShadowOffset:=ShadowOffset, DefaultShadowOpacity := DefaultTextShadowOpacity}
    EquipPet_1_ChangeSlotButton : button_regular = button_regular{}
    EquipPet_1_Text_PackageA_LV : text_block = text_block{DefaultTextColor := White, DefaultShadowOffset:=ShadowOffset, DefaultShadowOpacity := DefaultTextShadowOpacity}
    EquipPet_1_Text_PackageB_LV : text_block = text_block{DefaultTextColor := White, DefaultShadowOffset:=ShadowOffset, DefaultShadowOpacity := DefaultTextShadowOpacity} 

        #TOURNAMENT
    var EquipPet_1_TextureBlock_PetImages : texture_block = texture_block{DefaultImage:=Images.EmptyTexture,DefaultDesiredSize:=vector2{X:=75.0,Y:=75.0}}
    var EquipPet_1_TextBlock_Purity : text_block = text_block{DefaultTextColor := CurrentTextColor, DefaultShadowOffset:=ShadowOffset, DefaultShadowOpacity := DefaultTextShadowOpacity}
    var EquipPet_1_TextBlock_Class : text_block = text_block{DefaultTextColor := CurrentTextColor, DefaultShadowOffset:=ShadowOffset, DefaultShadowOpacity := DefaultTextShadowOpacity}
    var EquipPet_1_TextBlock_PackA : text_block = text_block{DefaultTextColor := CurrentTextColor, DefaultShadowOffset:=ShadowOffset, DefaultShadowOpacity := DefaultTextShadowOpacity}
    var EquipPet_1_TextBlock_PackB : text_block = text_block{DefaultTextColor := CurrentTextColor, DefaultShadowOffset:=ShadowOffset, DefaultShadowOpacity := DefaultTextShadowOpacity}
        #Package A
    var EquipPet_1_TextBlock_Description_SpecialPerkA : text_block = text_block{DefaultTextColor := Pink, DefaultShadowOffset:=ShadowOffset, DefaultShadowOpacity := DefaultTextShadowOpacity}
    var EquipPet_1_TextBlock_Current_SpecialPerkA : text_block = text_block{DefaultTextColor := CurrentTextColor, DefaultShadowOffset:=ShadowOffset, DefaultShadowOpacity := DefaultTextShadowOpacity}
    var EquipPet_1_TextBlock_Current_PlayerPickDamage : text_block = text_block{DefaultTextColor := CurrentTextColor, DefaultShadowOffset:=ShadowOffset, DefaultShadowOpacity := DefaultTextShadowOpacity}
    var EquipPet_1_TextBlock_Current_PlayerWeakPoint : text_block = text_block{DefaultTextColor := CurrentTextColor, DefaultShadowOffset:=ShadowOffset, DefaultShadowOpacity := DefaultTextShadowOpacity}
        #Package B
    var EquipPet_1_TextBlock_Description_SpecialPerkB : text_block = text_block{DefaultTextColor := Pink, DefaultShadowOffset:=ShadowOffset, DefaultShadowOpacity := DefaultTextShadowOpacity}
    var EquipPet_1_TextBlock_Current_SpecialPerkB : text_block = text_block{DefaultTextColor := CurrentTextColor, DefaultShadowOffset:=ShadowOffset, DefaultShadowOpacity := DefaultTextShadowOpacity}
    var EquipPet_1_TextBlock_Current_PlayerPickEfficiency : text_block = text_block{DefaultTextColor := CurrentTextColor, DefaultShadowOffset:=ShadowOffset, DefaultShadowOpacity := DefaultTextShadowOpacity}
    var EquipPet_1_TextBlock_Current_PlayerPickCriticalDamage : text_block = text_block{DefaultTextColor := CurrentTextColor, DefaultShadowOffset:=ShadowOffset, DefaultShadowOpacity := DefaultTextShadowOpacity}
    
    #PET 2
    EquipPet_2_SlotInfo : text_block = text_block{DefaultTextColor:=White, DefaultShadowOffset:=ShadowOffset, DefaultShadowOpacity := DefaultTextShadowOpacity}
    EquipPet_2_ChangeSlotButton : button_regular = button_regular{}
    EquipPet_2_Text_PackageA_LV : text_block = text_block{DefaultTextColor := White, DefaultShadowOffset:=ShadowOffset, DefaultShadowOpacity := DefaultTextShadowOpacity}
    EquipPet_2_Text_PackageB_LV : text_block = text_block{DefaultTextColor := White, DefaultShadowOffset:=ShadowOffset, DefaultShadowOpacity := DefaultTextShadowOpacity} 
    EquipPet_2_TextBlock_PackA : text_block = text_block{DefaultTextColor := CurrentTextColor, DefaultShadowOffset:=ShadowOffset, DefaultShadowOpacity := DefaultTextShadowOpacity}
    EquipPet_2_TextBlock_PackB : text_block = text_block{DefaultTextColor := CurrentTextColor, DefaultShadowOffset:=ShadowOffset, DefaultShadowOpacity := DefaultTextShadowOpacity}
        #TOURNAMENT
    var EquipPet_2_TextureBlock_PetImages : texture_block = texture_block{DefaultImage:=Images.EmptyTexture,DefaultDesiredSize:=vector2{X:=75.0,Y:=75.0}}
    var EquipPet_2_TextBlock_Purity : text_block = text_block{DefaultTextColor := CurrentTextColor, DefaultShadowOffset:=ShadowOffset, DefaultShadowOpacity := DefaultTextShadowOpacity}
    var EquipPet_2_TextBlock_Class : text_block = text_block{DefaultTextColor := CurrentTextColor, DefaultShadowOffset:=ShadowOffset, DefaultShadowOpacity := DefaultTextShadowOpacity}

        #Package A
    var EquipPet_2_TextBlock_Description_SpecialPerkA : text_block = text_block{DefaultTextColor := Pink, DefaultShadowOffset:=ShadowOffset, DefaultShadowOpacity := DefaultTextShadowOpacity}
    var EquipPet_2_TextBlock_Current_SpecialPerkA : text_block = text_block{DefaultTextColor := CurrentTextColor, DefaultShadowOffset:=ShadowOffset, DefaultShadowOpacity := DefaultTextShadowOpacity}
    var EquipPet_2_TextBlock_Current_PlayerPickDamage : text_block = text_block{DefaultTextColor := CurrentTextColor, DefaultShadowOffset:=ShadowOffset, DefaultShadowOpacity := DefaultTextShadowOpacity}
    var EquipPet_2_TextBlock_Current_PlayerWeakPoint : text_block = text_block{DefaultTextColor := CurrentTextColor, DefaultShadowOffset:=ShadowOffset, DefaultShadowOpacity := DefaultTextShadowOpacity}
        #Package B
    var EquipPet_2_TextBlock_Description_SpecialPerkB : text_block = text_block{DefaultTextColor := Pink, DefaultShadowOffset:=ShadowOffset, DefaultShadowOpacity := DefaultTextShadowOpacity}
    var EquipPet_2_TextBlock_Current_SpecialPerkB : text_block = text_block{DefaultTextColor := CurrentTextColor, DefaultShadowOffset:=ShadowOffset, DefaultShadowOpacity := DefaultTextShadowOpacity}
    var EquipPet_2_TextBlock_Current_PlayerPickEfficiency : text_block = text_block{DefaultTextColor := CurrentTextColor, DefaultShadowOffset:=ShadowOffset, DefaultShadowOpacity := DefaultTextShadowOpacity}
    var EquipPet_2_TextBlock_Current_PlayerPickCriticalDamage : text_block = text_block{DefaultTextColor := CurrentTextColor, DefaultShadowOffset:=ShadowOffset, DefaultShadowOpacity := DefaultTextShadowOpacity}

    EquipPet_CloseButton : button_loud = button_loud{}

#CREATE PET
#########
    var CreatePet_MaybeCanvas : ?canvas = false
    var CreatePet_UIIsDisplayed : logic = false
    var CreatedPetTierWarning : logic = false

    var ConfirmationToCreate : logic = false

    var CreatePet_PetToCreate_CurrentPurity : int = 1
    var CreatePet_PetToCreate_CurrentTier : int = 0
    var CreatePet_PetToCreate_CurrentClass : int = 0
    var CreatePet_TotalCost : float = 0.0
    
    CreatePet_CreatePetButton : button_quiet = button_quiet{}
    CreatePet_ChangeClassButton : button_quiet = button_quiet{}

    CreatePet_Add1PurityButton : button_quiet  = button_quiet{}
    CreatePet_Add5PurityButton : button_quiet  = button_quiet{}
    CreatePet_Add25PurityButton : button_quiet = button_quiet{}

    CreatePet_Remove1PurityButton : button_quiet = button_quiet{}
    CreatePet_Remove5PurityButton : button_quiet = button_quiet{}
    CreatePet_Remove25PurityButton : button_quiet = button_quiet{}

    CreatePet_Text_PackageA_LV : text_block = text_block{DefaultTextColor := White, DefaultShadowOffset:=ShadowOffset, DefaultShadowOpacity := DefaultTextShadowOpacity}
    CreatePet_Text_PackageB_LV : text_block = text_block{DefaultTextColor := White, DefaultShadowOffset:=ShadowOffset, DefaultShadowOpacity := DefaultTextShadowOpacity} 

    #TOURNAMENT
    var CreatePet_TextureBlock_PetImages : texture_block = texture_block{DefaultImage:=Images.EmptyTexture,DefaultDesiredSize:=vector2{X:=75.0,Y:=75.0}}
    var CreatePet_TextBlock_Purity : text_block = text_block{DefaultTextColor := CurrentTextColor, DefaultShadowOffset:=ShadowOffset, DefaultShadowOpacity := DefaultTextShadowOpacity}
    var CreatePet_TextBlock_Class : text_block = text_block{DefaultTextColor := CurrentTextColor, DefaultShadowOffset:=ShadowOffset, DefaultShadowOpacity := DefaultTextShadowOpacity}
    var CreatePet_TextBlock_PackA : text_block = text_block{DefaultTextColor := CurrentTextColor, DefaultShadowOffset:=ShadowOffset, DefaultShadowOpacity := DefaultTextShadowOpacity}
    var CreatePet_TextBlock_PackB : text_block = text_block{DefaultTextColor := CurrentTextColor, DefaultShadowOffset:=ShadowOffset, DefaultShadowOpacity := DefaultTextShadowOpacity}

    #Package A
    var CreatePet_TextBlock_Current_TournamentHealth : text_block = text_block{DefaultTextColor := CurrentTextColor, DefaultShadowOffset:=ShadowOffset, DefaultShadowOpacity := DefaultTextShadowOpacity}
    var CreatePet_TextBlock_Current_TournamentDamage : text_block = text_block{DefaultTextColor := CurrentTextColor, DefaultShadowOffset:=ShadowOffset, DefaultShadowOpacity := DefaultTextShadowOpacity}
    var CreatePet_TextBlock_Current_TournamentSpeed : text_block = text_block{DefaultTextColor := CurrentTextColor, DefaultShadowOffset:=ShadowOffset, DefaultShadowOpacity := DefaultTextShadowOpacity}

    #Package B
    var CreatePet_TextBlock_Current_TournamentActionSpeed : text_block = text_block{DefaultTextColor := CurrentTextColor, DefaultShadowOffset:=ShadowOffset, DefaultShadowOpacity := DefaultTextShadowOpacity}
    var CreatePet_TextBlock_Current_TournamentRange : text_block = text_block{DefaultTextColor := CurrentTextColor, DefaultShadowOffset:=ShadowOffset, DefaultShadowOpacity := DefaultTextShadowOpacity}
    var CreatePet_TextBlock_Current_TournamentHealingPerk : text_block = text_block{DefaultTextColor := CurrentTextColor, DefaultShadowOffset:=ShadowOffset, DefaultShadowOpacity := DefaultTextShadowOpacity}

    #PLAYER STATS
        #Package A
    var CreatePet_TextBlock_Description_SpecialPerkA : text_block = text_block{DefaultTextColor := Pink, DefaultShadowOffset:=ShadowOffset, DefaultShadowOpacity := DefaultTextShadowOpacity}

    var CreatePet_TextBlock_Current_SpecialPerkA : text_block = text_block{DefaultTextColor := CurrentTextColor, DefaultShadowOffset:=ShadowOffset, DefaultShadowOpacity := DefaultTextShadowOpacity}
    var CreatePet_TextBlock_Current_PlayerPickDamage : text_block = text_block{DefaultTextColor := CurrentTextColor, DefaultShadowOffset:=ShadowOffset, DefaultShadowOpacity := DefaultTextShadowOpacity}
    var CreatePet_TextBlock_Current_PlayerWeakPoint : text_block = text_block{DefaultTextColor := CurrentTextColor, DefaultShadowOffset:=ShadowOffset, DefaultShadowOpacity := DefaultTextShadowOpacity}
        #Package B
    var CreatePet_TextBlock_Description_SpecialPerkB : text_block = text_block{DefaultTextColor := Pink, DefaultShadowOffset:=ShadowOffset, DefaultShadowOpacity := DefaultTextShadowOpacity}

    var CreatePet_TextBlock_Current_SpecialPerkB : text_block = text_block{DefaultTextColor := CurrentTextColor, DefaultShadowOffset:=ShadowOffset, DefaultShadowOpacity := DefaultTextShadowOpacity}
    var CreatePet_TextBlock_Current_PlayerPickEfficiency : text_block = text_block{DefaultTextColor := CurrentTextColor, DefaultShadowOffset:=ShadowOffset, DefaultShadowOpacity := DefaultTextShadowOpacity}
    var CreatePet_TextBlock_Current_PlayerPickCriticalDamage : text_block = text_block{DefaultTextColor := CurrentTextColor, DefaultShadowOffset:=ShadowOffset, DefaultShadowOpacity := DefaultTextShadowOpacity}

    CreatePet_SlotsAvailable_TextBlock : text_block = text_block{DefaultTextColor:=White, DefaultShadowOffset:=ShadowOffset, DefaultShadowOpacity := DefaultTextShadowOpacity}
    CreatePet_TotalInventoryMutantMaterial_TextBlock : text_block = text_block{DefaultTextColor:=MutantMaterialTextColor, DefaultShadowOffset:=ShadowOffset, DefaultShadowOpacity := DefaultTextShadowOpacity}
    CreatePet_TotalCostMutantMaterial_TextBlock : text_block = text_block{DefaultTextColor:=MutantMaterialTextColor, DefaultShadowOffset:=ShadowOffset, DefaultShadowOpacity := DefaultTextShadowOpacity}
    CreatePet_CloseButton : button_loud = button_loud{}
#RECICLE
#########

    var IsRecycleReady : logic = false
    var ConfirmationToRecycle : int = 0

    var RecyclePet_MaybeCanvas : ?canvas = false
    var RecyclePet_UIIsDisplayed : logic = false
    
    #Buttons
    var RecyclePet_CurrentOption : int = 0
    
    RecyclePet_ExecuteRecycleButton : button_regular = button_regular{}
    RecyclePet_NavigateRightButton : button_regular = button_regular{}
    RecyclePet_NavigateLeftButton : button_regular = button_regular{}
    
    RecyclePet_CancelButton : button_loud = button_loud{}
    
    #Box0
    RecyclePet_TextureBlock_PetImages_Box0 : texture_block = texture_block{DefaultImage:=Images.EmptyTexture,DefaultDesiredSize:=PetsSelectorImageSize}
    RecyclePet_TextBlock_Classes_Box0 : text_block = text_block{DefaultTextColor := White, DefaultShadowOffset:=ShadowOffset, DefaultShadowOpacity := DefaultTextShadowOpacity}
    RecyclePet_TextBlock_PackA_Box0 : text_block = text_block{DefaultTextColor := White, DefaultShadowOffset:=ShadowOffset, DefaultShadowOpacity := DefaultTextShadowOpacity}
    RecyclePet_TextBlock_PackB_Box0 : text_block = text_block{DefaultTextColor := White, DefaultShadowOffset:=ShadowOffset, DefaultShadowOpacity := DefaultTextShadowOpacity}
    RecyclePet_TextBlock_Purity_Box0 : text_block = text_block{DefaultTextColor := White, DefaultShadowOffset:=ShadowOffset, DefaultShadowOpacity := DefaultTextShadowOpacity}
    RecyclePet_TextBlock_Tier_Box0 : text_block = text_block{DefaultTextColor := White, DefaultShadowOffset:=ShadowOffset, DefaultShadowOpacity := DefaultTextShadowOpacity}
    
    #Box1
    RecyclePet_TextureBlock_PetImages_Box1 : texture_block = texture_block{DefaultImage:=Images.EmptyTexture,DefaultDesiredSize:=PetsSelectorImageSize}
    RecyclePet_TextBlock_Classes_Box1 : text_block = text_block{DefaultTextColor := White, DefaultShadowOffset:=ShadowOffset, DefaultShadowOpacity := DefaultTextShadowOpacity}
    RecyclePet_TextBlock_PackA_Box1 : text_block = text_block{DefaultTextColor := White, DefaultShadowOffset:=ShadowOffset, DefaultShadowOpacity := DefaultTextShadowOpacity}
    RecyclePet_TextBlock_PackB_Box1 : text_block = text_block{DefaultTextColor := White, DefaultShadowOffset:=ShadowOffset, DefaultShadowOpacity := DefaultTextShadowOpacity}
    RecyclePet_TextBlock_Purity_Box1 : text_block = text_block{DefaultTextColor := White, DefaultShadowOffset:=ShadowOffset, DefaultShadowOpacity := DefaultTextShadowOpacity}
    RecyclePet_TextBlock_Tier_Box1 : text_block = text_block{DefaultTextColor := White, DefaultShadowOffset:=ShadowOffset, DefaultShadowOpacity := DefaultTextShadowOpacity}
    
    #Box2
    RecyclePet_TextureBlock_PetImages_Box2 : texture_block = texture_block{DefaultImage:=Images.EmptyTexture,DefaultDesiredSize:=PetsSelectorImageSize}
    RecyclePet_TextBlock_Classes_Box2 : text_block = text_block{DefaultTextColor := White, DefaultShadowOffset:=ShadowOffset, DefaultShadowOpacity := DefaultTextShadowOpacity}
    RecyclePet_TextBlock_PackA_Box2 : text_block = text_block{DefaultTextColor := White, DefaultShadowOffset:=ShadowOffset, DefaultShadowOpacity := DefaultTextShadowOpacity}
    RecyclePet_TextBlock_PackB_Box2 : text_block = text_block{DefaultTextColor := White, DefaultShadowOffset:=ShadowOffset, DefaultShadowOpacity := DefaultTextShadowOpacity}
    RecyclePet_TextBlock_Purity_Box2 : text_block = text_block{DefaultTextColor := White, DefaultShadowOffset:=ShadowOffset, DefaultShadowOpacity := DefaultTextShadowOpacity}
    RecyclePet_TextBlock_Tier_Box2 : text_block = text_block{DefaultTextColor := White, DefaultShadowOffset:=ShadowOffset, DefaultShadowOpacity := DefaultTextShadowOpacity}
    
    #Box3
    RecyclePet_TextureBlock_PetImages_Box3 : texture_block = texture_block{DefaultImage:=Images.EmptyTexture,DefaultDesiredSize:=PetsSelectorImageSize}
    RecyclePet_TextBlock_Classes_Box3 : text_block = text_block{DefaultTextColor := White, DefaultShadowOffset:=ShadowOffset, DefaultShadowOpacity := DefaultTextShadowOpacity}
    RecyclePet_TextBlock_PackA_Box3 : text_block = text_block{DefaultTextColor := White, DefaultShadowOffset:=ShadowOffset, DefaultShadowOpacity := DefaultTextShadowOpacity}
    RecyclePet_TextBlock_PackB_Box3 : text_block = text_block{DefaultTextColor := White, DefaultShadowOffset:=ShadowOffset, DefaultShadowOpacity := DefaultTextShadowOpacity}
    RecyclePet_TextBlock_Purity_Box3 : text_block = text_block{DefaultTextColor := White, DefaultShadowOffset:=ShadowOffset, DefaultShadowOpacity := DefaultTextShadowOpacity}
    RecyclePet_TextBlock_Tier_Box3 : text_block = text_block{DefaultTextColor := White, DefaultShadowOffset:=ShadowOffset, DefaultShadowOpacity := DefaultTextShadowOpacity}
    
    #All Arrays
    var AllRecyclePet_TextureBlock_PetImages : []texture_block = array{}
    var AllRecyclePet_TextBlock_Classes : []text_block = array{}
    var AllRecyclePet_TextBlock_PackA : []text_block = array{}
    var AllRecyclePet_TextBlock_PackB : []text_block = array{}
    var AllRecyclePet_TextBlock_Purity : []text_block = array{}
    var AllRecyclePet_TextBlock_Tier : []text_block = array{}
    
    RecyclePet_Text_MuntantMaterialRecicled : text_block = text_block{DefaultTextColor := MutantMaterialTextColor, DefaultShadowOffset:=ShadowOffset, DefaultShadowOpacity := DefaultTextShadowOpacity}
    RecyclePet_Text_CoalRecicled : text_block = text_block{DefaultTextColor := CoalTextColor, DefaultShadowOffset:=ShadowOffset, DefaultShadowOpacity := DefaultTextShadowOpacity}
    
    RecyclePet_Text_PackageA_LV : text_block = text_block{DefaultTextColor := White, DefaultShadowOffset:=ShadowOffset, DefaultShadowOpacity := DefaultTextShadowOpacity}
    RecyclePet_Text_PackageB_LV : text_block = text_block{DefaultTextColor := White, DefaultShadowOffset:=ShadowOffset, DefaultShadowOpacity := DefaultTextShadowOpacity}
    
    #TOURNAMENT
    var RecyclePet_Selected_TextureBlock_PetImages : texture_block = texture_block{DefaultImage:=Images.EmptyTexture,DefaultDesiredSize:=vector2{X:=75.0,Y:=75.0}}
    var RecyclePet_Selected_TextBlock_Purity : text_block = text_block{DefaultTextColor := CurrentTextColor, DefaultShadowOffset:=ShadowOffset, DefaultShadowOpacity := DefaultTextShadowOpacity}
    var RecyclePet_Selected_TextBlock_Class : text_block = text_block{DefaultTextColor := CurrentTextColor, DefaultShadowOffset:=ShadowOffset, DefaultShadowOpacity := DefaultTextShadowOpacity}
    var RecyclePet_Selected_TextBlock_PackA : text_block = text_block{DefaultTextColor := CurrentTextColor, DefaultShadowOffset:=ShadowOffset, DefaultShadowOpacity := DefaultTextShadowOpacity}
    var RecyclePet_Selected_TextBlock_PackB : text_block = text_block{DefaultTextColor := CurrentTextColor, DefaultShadowOffset:=ShadowOffset, DefaultShadowOpacity := DefaultTextShadowOpacity}
    
    #Package A
    var RecyclePet_Selected_TextBlock_Current_TournamentHealth : text_block = text_block{DefaultTextColor := CurrentTextColor, DefaultShadowOffset:=ShadowOffset, DefaultShadowOpacity := DefaultTextShadowOpacity}
    var RecyclePet_Selected_TextBlock_Current_TournamentDamage : text_block = text_block{DefaultTextColor := CurrentTextColor, DefaultShadowOffset:=ShadowOffset, DefaultShadowOpacity := DefaultTextShadowOpacity}
    var RecyclePet_Selected_TextBlock_Current_TournamentSpeed : text_block = text_block{DefaultTextColor := CurrentTextColor, DefaultShadowOffset:=ShadowOffset, DefaultShadowOpacity := DefaultTextShadowOpacity}
    
    #Package B
    var RecyclePet_Selected_TextBlock_Current_TournamentActionSpeed : text_block = text_block{DefaultTextColor := CurrentTextColor, DefaultShadowOffset:=ShadowOffset, DefaultShadowOpacity := DefaultTextShadowOpacity}
    var RecyclePet_Selected_TextBlock_Current_TournamentRange : text_block = text_block{DefaultTextColor := CurrentTextColor, DefaultShadowOffset:=ShadowOffset, DefaultShadowOpacity := DefaultTextShadowOpacity}
    var RecyclePet_Selected_TextBlock_Current_TournamentHealingPerk : text_block = text_block{DefaultTextColor := CurrentTextColor, DefaultShadowOffset:=ShadowOffset, DefaultShadowOpacity := DefaultTextShadowOpacity}
    
    #PLAYER STATS
        #Package A
    var RecyclePet_Selected_TextBlock_Description_SpecialPerkA : text_block = text_block{DefaultTextColor := Pink, DefaultShadowOffset:=ShadowOffset, DefaultShadowOpacity := DefaultTextShadowOpacity}
    
    var RecyclePet_Selected_TextBlock_Current_SpecialPerkA : text_block = text_block{DefaultTextColor := CurrentTextColor, DefaultShadowOffset:=ShadowOffset, DefaultShadowOpacity := DefaultTextShadowOpacity}
    var RecyclePet_Selected_TextBlock_Current_PlayerPickDamage : text_block = text_block{DefaultTextColor := CurrentTextColor, DefaultShadowOffset:=ShadowOffset, DefaultShadowOpacity := DefaultTextShadowOpacity}
    var RecyclePet_Selected_TextBlock_Current_PlayerWeakPoint : text_block = text_block{DefaultTextColor := CurrentTextColor, DefaultShadowOffset:=ShadowOffset, DefaultShadowOpacity := DefaultTextShadowOpacity}
    
        #Package B
    var RecyclePet_Selected_TextBlock_Description_SpecialPerkB : text_block = text_block{DefaultTextColor := Pink, DefaultShadowOffset:=ShadowOffset, DefaultShadowOpacity := DefaultTextShadowOpacity}
    
    var RecyclePet_Selected_TextBlock_Current_SpecialPerkB : text_block = text_block{DefaultTextColor := CurrentTextColor, DefaultShadowOffset:=ShadowOffset, DefaultShadowOpacity := DefaultTextShadowOpacity}
    var RecyclePet_Selected_TextBlock_Current_PlayerPickEfficiency : text_block = text_block{DefaultTextColor := CurrentTextColor, DefaultShadowOffset:=ShadowOffset, DefaultShadowOpacity := DefaultTextShadowOpacity}
    var RecyclePet_Selected_TextBlock_Current_PlayerPickCriticalDamage : text_block = text_block{DefaultTextColor := CurrentTextColor, DefaultShadowOffset:=ShadowOffset, DefaultShadowOpacity := DefaultTextShadowOpacity}



#MIX MENU
#########
    var MixPets_MaybeCanvas : ?canvas = false
    var MixPets_UIIsDisplayed : logic = false

    MixPets_MixTitle : text_block = text_block{DefaultTextColor:=Purple, DefaultShadowOffset:=ShadowOffset, DefaultShadowOpacity := DefaultTextShadowOpacity}

    var IsFusionReady : logic = false
    var FusionedPet : pet_data_table = pet_data_table{}
    var FusionedPetTierWarning : logic = false
    
    var ConfirmationToMix : logic = false

    var AllMixPets_PetsToMix : []int = array{-1,-1}

    MixPets_MixButton : button_regular = button_regular{}
    MixPets_Button_Box0 : button_regular = button_regular{}
    MixPets_Button_Box1 : button_regular = button_regular{}
    MixPets_CloseButton : button_loud = button_loud{}

    CalculatingMixtureTextureBlock : texture_block = texture_block{DefaultImage:=Images.EmptyTexture,DefaultDesiredSize:=vector2{X:=1300.0,Y:=705.0}}

        #Box0
    MixPets_TextureBlock_PetImages_Box0 : texture_block = texture_block{DefaultImage:=Images.EmptyTexture,DefaultDesiredSize:=PetsSelectorImageSize}
    MixPets_TextBlock_Classes_Box0 : text_block = text_block{DefaultTextColor := White, DefaultShadowOffset:=ShadowOffset, DefaultShadowOpacity := DefaultTextShadowOpacity}
    MixPets_TextBlock_PackA_Box0 : text_block = text_block{DefaultTextColor:=White, DefaultShadowOffset:=ShadowOffset, DefaultShadowOpacity := DefaultTextShadowOpacity}
    MixPets_TextBlock_PackB_Box0 : text_block = text_block{DefaultTextColor:=White, DefaultShadowOffset:=ShadowOffset, DefaultShadowOpacity := DefaultTextShadowOpacity}
    MixPets_TextBlock_Purity_Box0 : text_block = text_block{DefaultTextColor := White, DefaultShadowOffset:=ShadowOffset, DefaultShadowOpacity := DefaultTextShadowOpacity}
    MixPets_TextBlock_Tier_Box0 : text_block = text_block{DefaultTextColor := White, DefaultShadowOffset:=ShadowOffset, DefaultShadowOpacity := DefaultTextShadowOpacity}

        #Box1
    MixPets_TextureBlock_PetImages_Box1 : texture_block = texture_block{DefaultImage:=Images.EmptyTexture,DefaultDesiredSize:=PetsSelectorImageSize}
    MixPets_TextBlock_Classes_Box1 : text_block = text_block{DefaultTextColor := White, DefaultShadowOffset:=ShadowOffset, DefaultShadowOpacity := DefaultTextShadowOpacity}
    MixPets_TextBlock_PackA_Box1 : text_block = text_block{DefaultTextColor:=White, DefaultShadowOffset:=ShadowOffset, DefaultShadowOpacity := DefaultTextShadowOpacity}
    MixPets_TextBlock_PackB_Box1 : text_block = text_block{DefaultTextColor:=White, DefaultShadowOffset:=ShadowOffset, DefaultShadowOpacity := DefaultTextShadowOpacity}
    MixPets_TextBlock_Purity_Box1 : text_block = text_block{DefaultTextColor := White, DefaultShadowOffset:=ShadowOffset, DefaultShadowOpacity := DefaultTextShadowOpacity}
    MixPets_TextBlock_Tier_Box1 : text_block = text_block{DefaultTextColor := White, DefaultShadowOffset:=ShadowOffset, DefaultShadowOpacity := DefaultTextShadowOpacity}

        #All Arrays
    var AllMixPets_TextureBlock_PetImages : []texture_block = array{}
    var AllMixPets_TextBlock_Classes : []text_block = array{}
    var AllMixPets_TextBlock_PackA : []text_block = array{}
    var AllMixPets_TextBlock_PackB : []text_block = array{}
    var AllMixPets_TextBlock_Purity : []text_block = array{}
    var AllMixPets_TextBlock_Tier : []text_block = array{}


        #TOURNAMENT
    MixPets_Text_PackageA_LV : text_block = text_block{DefaultTextColor := White, DefaultShadowOffset:=ShadowOffset, DefaultShadowOpacity := DefaultTextShadowOpacity}
    MixPets_Text_PackageB_LV : text_block = text_block{DefaultTextColor := White, DefaultShadowOffset:=ShadowOffset, DefaultShadowOpacity := DefaultTextShadowOpacity}

    var MixPets_Selected_TextureBlock_PetImages : texture_block = texture_block{DefaultImage:=Images.EmptyTexture,DefaultDesiredSize:=vector2{X:=75.0,Y:=75.0}}
    var MixPets_Selected_TextBlock_Purity : text_block = text_block{DefaultTextColor := CurrentTextColor, DefaultShadowOffset:=ShadowOffset, DefaultShadowOpacity := DefaultTextShadowOpacity}
    var MixPets_Selected_TextBlock_Class : text_block = text_block{DefaultTextColor := CurrentTextColor, DefaultShadowOffset:=ShadowOffset, DefaultShadowOpacity := DefaultTextShadowOpacity}
    var MixPets_Selected_TextBlock_PackA : text_block = text_block{DefaultTextColor := CurrentTextColor, DefaultShadowOffset:=ShadowOffset, DefaultShadowOpacity := DefaultTextShadowOpacity}
    var MixPets_Selected_TextBlock_PackB : text_block = text_block{DefaultTextColor := CurrentTextColor, DefaultShadowOffset:=ShadowOffset, DefaultShadowOpacity := DefaultTextShadowOpacity}

        #Package A
    var MixPets_Selected_TextBlock_Current_TournamentHealth : text_block = text_block{DefaultTextColor := CurrentTextColor, DefaultShadowOffset:=ShadowOffset, DefaultShadowOpacity := DefaultTextShadowOpacity}
    var MixPets_Selected_TextBlock_Current_TournamentDamage : text_block = text_block{DefaultTextColor := CurrentTextColor, DefaultShadowOffset:=ShadowOffset, DefaultShadowOpacity := DefaultTextShadowOpacity}
    var MixPets_Selected_TextBlock_Current_TournamentSpeed : text_block = text_block{DefaultTextColor := CurrentTextColor, DefaultShadowOffset:=ShadowOffset, DefaultShadowOpacity := DefaultTextShadowOpacity}

        #Package B
    var MixPets_Selected_TextBlock_Current_TournamentActionSpeed : text_block = text_block{DefaultTextColor := CurrentTextColor, DefaultShadowOffset:=ShadowOffset, DefaultShadowOpacity := DefaultTextShadowOpacity}
    var MixPets_Selected_TextBlock_Current_TournamentRange : text_block = text_block{DefaultTextColor := CurrentTextColor, DefaultShadowOffset:=ShadowOffset, DefaultShadowOpacity := DefaultTextShadowOpacity}
    var MixPets_Selected_TextBlock_Current_TournamentHealingPerk : text_block = text_block{DefaultTextColor := CurrentTextColor, DefaultShadowOffset:=ShadowOffset, DefaultShadowOpacity := DefaultTextShadowOpacity}

        #PLAYER STATS
        #Package A
    var MixPets_Selected_TextBlock_Description_SpecialPerkA : text_block = text_block{DefaultTextColor := Pink, DefaultShadowOffset:=ShadowOffset, DefaultShadowOpacity := DefaultTextShadowOpacity}
    var MixPets_Selected_TextBlock_Current_SpecialPerkA : text_block = text_block{DefaultTextColor := CurrentTextColor, DefaultShadowOffset:=ShadowOffset, DefaultShadowOpacity := DefaultTextShadowOpacity}
    var MixPets_Selected_TextBlock_Current_PlayerPickDamage : text_block = text_block{DefaultTextColor := CurrentTextColor, DefaultShadowOffset:=ShadowOffset, DefaultShadowOpacity := DefaultTextShadowOpacity}
    var MixPets_Selected_TextBlock_Current_PlayerWeakPoint : text_block = text_block{DefaultTextColor := CurrentTextColor, DefaultShadowOffset:=ShadowOffset, DefaultShadowOpacity := DefaultTextShadowOpacity}

        #Package B
    var MixPets_Selected_TextBlock_Description_SpecialPerkB : text_block = text_block{DefaultTextColor := Pink, DefaultShadowOffset:=ShadowOffset, DefaultShadowOpacity := DefaultTextShadowOpacity}
    var MixPets_Selected_TextBlock_Current_SpecialPerkB : text_block = text_block{DefaultTextColor := CurrentTextColor, DefaultShadowOffset:=ShadowOffset, DefaultShadowOpacity := DefaultTextShadowOpacity}
    var MixPets_Selected_TextBlock_Current_PlayerPickEfficiency : text_block = text_block{DefaultTextColor := CurrentTextColor, DefaultShadowOffset:=ShadowOffset, DefaultShadowOpacity := DefaultTextShadowOpacity}
    var MixPets_Selected_TextBlock_Current_PlayerPickCriticalDamage : text_block = text_block{DefaultTextColor := CurrentTextColor, DefaultShadowOffset:=ShadowOffset, DefaultShadowOpacity := DefaultTextShadowOpacity}


#GENERAL PET SELECTOR
#########
    var GeneralSelector_OpenedFor : int = -1
    var GeneralSelector_UISlotSelected : int = -1
    var GeneralSelector_IsCancel : logic = false
    var GeneralSelector_FromSubmenu : int = -1

    var GeneralSelector_MaybeCanvas : ?canvas = false
    var GeneralSelector_UIIsDisplayed : logic = false
    
    #Buttons
    var GeneralSelector_CurrentOption : int = 0

    GeneralSelector_SelectButton : button_regular = button_regular{}
    GeneralSelector_NavigateRightButton : button_regular = button_regular{}
    GeneralSelector_NavigateLeftButton : button_regular = button_regular{}

    GeneralSelector_CancelButton : button_loud = button_loud{}

    #Box0
    GeneralSelector_TextureBlock_PetImages_Box0 : texture_block = texture_block{DefaultImage:=Images.EmptyTexture,DefaultDesiredSize:=PetsSelectorImageSize}
    GeneralSelector_TextBlock_Classes_Box0 : text_block = text_block{DefaultTextColor := White, DefaultShadowOffset:=ShadowOffset, DefaultShadowOpacity := DefaultTextShadowOpacity}
    GeneralSelector_TextBlock_PackA_Box0 : text_block = text_block{DefaultTextColor:=White, DefaultShadowOffset:=ShadowOffset, DefaultShadowOpacity := DefaultTextShadowOpacity}
    GeneralSelector_TextBlock_PackB_Box0 : text_block = text_block{DefaultTextColor:=White, DefaultShadowOffset:=ShadowOffset, DefaultShadowOpacity := DefaultTextShadowOpacity}
    GeneralSelector_TextBlock_Purity_Box0 : text_block = text_block{DefaultTextColor := White, DefaultShadowOffset:=ShadowOffset, DefaultShadowOpacity := DefaultTextShadowOpacity}
    GeneralSelector_TextBlock_Tier_Box0 : text_block = text_block{DefaultTextColor := White, DefaultShadowOffset:=ShadowOffset, DefaultShadowOpacity := DefaultTextShadowOpacity}

    #Box1
    GeneralSelector_TextureBlock_PetImages_Box1 : texture_block = texture_block{DefaultImage:=Images.EmptyTexture,DefaultDesiredSize:=PetsSelectorImageSize}
    GeneralSelector_TextBlock_Classes_Box1 : text_block = text_block{DefaultTextColor := White, DefaultShadowOffset:=ShadowOffset, DefaultShadowOpacity := DefaultTextShadowOpacity}
    GeneralSelector_TextBlock_PackA_Box1 : text_block = text_block{DefaultTextColor:=White, DefaultShadowOffset:=ShadowOffset, DefaultShadowOpacity := DefaultTextShadowOpacity}
    GeneralSelector_TextBlock_PackB_Box1 : text_block = text_block{DefaultTextColor:=White, DefaultShadowOffset:=ShadowOffset, DefaultShadowOpacity := DefaultTextShadowOpacity}
    GeneralSelector_TextBlock_Purity_Box1 : text_block = text_block{DefaultTextColor := White, DefaultShadowOffset:=ShadowOffset, DefaultShadowOpacity := DefaultTextShadowOpacity}
    GeneralSelector_TextBlock_Tier_Box1 : text_block = text_block{DefaultTextColor := White, DefaultShadowOffset:=ShadowOffset, DefaultShadowOpacity := DefaultTextShadowOpacity}

    #Box2
    GeneralSelector_TextureBlock_PetImages_Box2 : texture_block = texture_block{DefaultImage:=Images.EmptyTexture,DefaultDesiredSize:=PetsSelectorImageSize}
    GeneralSelector_TextBlock_Classes_Box2 : text_block = text_block{DefaultTextColor := White, DefaultShadowOffset:=ShadowOffset, DefaultShadowOpacity := DefaultTextShadowOpacity}
    GeneralSelector_TextBlock_PackA_Box2 : text_block = text_block{DefaultTextColor:=White, DefaultShadowOffset:=ShadowOffset, DefaultShadowOpacity := DefaultTextShadowOpacity}
    GeneralSelector_TextBlock_PackB_Box2 : text_block = text_block{DefaultTextColor:=White, DefaultShadowOffset:=ShadowOffset, DefaultShadowOpacity := DefaultTextShadowOpacity}
    GeneralSelector_TextBlock_Purity_Box2 : text_block = text_block{DefaultTextColor := White, DefaultShadowOffset:=ShadowOffset, DefaultShadowOpacity := DefaultTextShadowOpacity}
    GeneralSelector_TextBlock_Tier_Box2 : text_block = text_block{DefaultTextColor := White, DefaultShadowOffset:=ShadowOffset, DefaultShadowOpacity := DefaultTextShadowOpacity}

    #Box3
    GeneralSelector_TextureBlock_PetImages_Box3 : texture_block = texture_block{DefaultImage:=Images.EmptyTexture,DefaultDesiredSize:=PetsSelectorImageSize}
    GeneralSelector_TextBlock_Classes_Box3 : text_block = text_block{DefaultTextColor := White, DefaultShadowOffset:=ShadowOffset, DefaultShadowOpacity := DefaultTextShadowOpacity}
    GeneralSelector_TextBlock_PackA_Box3 : text_block = text_block{DefaultTextColor:=White, DefaultShadowOffset:=ShadowOffset, DefaultShadowOpacity := DefaultTextShadowOpacity}
    GeneralSelector_TextBlock_PackB_Box3 : text_block = text_block{DefaultTextColor:=White, DefaultShadowOffset:=ShadowOffset, DefaultShadowOpacity := DefaultTextShadowOpacity}
    GeneralSelector_TextBlock_Purity_Box3 : text_block = text_block{DefaultTextColor := White, DefaultShadowOffset:=ShadowOffset, DefaultShadowOpacity := DefaultTextShadowOpacity}
    GeneralSelector_TextBlock_Tier_Box3 : text_block = text_block{DefaultTextColor := White, DefaultShadowOffset:=ShadowOffset, DefaultShadowOpacity := DefaultTextShadowOpacity}

    #All Arrays
    var AllGeneralSelector_TextureBlock_PetImages : []texture_block = array{}
    var AllGeneralSelector_TextBlock_Classes : []text_block = array{}
    var AllGeneralSelector_TextBlock_PackA : []text_block = array{}
    var AllGeneralSelector_TextBlock_PackB : []text_block = array{}
    var AllGeneralSelector_TextBlock_Purity : []text_block = array{}
    var AllGeneralSelector_TextBlock_Tier : []text_block = array{}

    GeneralSelector_Text_PackageA_LV : text_block = text_block{DefaultTextColor := White, DefaultShadowOffset:=ShadowOffset, DefaultShadowOpacity := DefaultTextShadowOpacity}
    GeneralSelector_Text_PackageB_LV : text_block = text_block{DefaultTextColor := White, DefaultShadowOffset:=ShadowOffset, DefaultShadowOpacity := DefaultTextShadowOpacity}

    #TOURNAMENT
    var GeneralSelector_Selected_TextureBlock_PetImages : texture_block = texture_block{DefaultImage:=Images.EmptyTexture,DefaultDesiredSize:=vector2{X:=75.0,Y:=75.0}}
    var GeneralSelector_Selected_TextBlock_Purity : text_block = text_block{DefaultTextColor := CurrentTextColor, DefaultShadowOffset:=ShadowOffset, DefaultShadowOpacity := DefaultTextShadowOpacity}
    var GeneralSelector_Selected_TextBlock_Class : text_block = text_block{DefaultTextColor := CurrentTextColor, DefaultShadowOffset:=ShadowOffset, DefaultShadowOpacity := DefaultTextShadowOpacity}
    var GeneralSelector_Selected_TextBlock_PackA : text_block = text_block{DefaultTextColor := CurrentTextColor, DefaultShadowOffset:=ShadowOffset, DefaultShadowOpacity := DefaultTextShadowOpacity}
    var GeneralSelector_Selected_TextBlock_PackB : text_block = text_block{DefaultTextColor := CurrentTextColor, DefaultShadowOffset:=ShadowOffset, DefaultShadowOpacity := DefaultTextShadowOpacity}

    #Package A
    var GeneralSelector_Selected_TextBlock_Current_TournamentHealth : text_block = text_block{DefaultTextColor := CurrentTextColor, DefaultShadowOffset:=ShadowOffset, DefaultShadowOpacity := DefaultTextShadowOpacity}
    var GeneralSelector_Selected_TextBlock_Current_TournamentDamage : text_block = text_block{DefaultTextColor := CurrentTextColor, DefaultShadowOffset:=ShadowOffset, DefaultShadowOpacity := DefaultTextShadowOpacity}
    var GeneralSelector_Selected_TextBlock_Current_TournamentSpeed : text_block = text_block{DefaultTextColor := CurrentTextColor, DefaultShadowOffset:=ShadowOffset, DefaultShadowOpacity := DefaultTextShadowOpacity}

    #Package B
    var GeneralSelector_Selected_TextBlock_Current_TournamentActionSpeed : text_block = text_block{DefaultTextColor := CurrentTextColor, DefaultShadowOffset:=ShadowOffset, DefaultShadowOpacity := DefaultTextShadowOpacity}
    var GeneralSelector_Selected_TextBlock_Current_TournamentRange : text_block = text_block{DefaultTextColor := CurrentTextColor, DefaultShadowOffset:=ShadowOffset, DefaultShadowOpacity := DefaultTextShadowOpacity}
    var GeneralSelector_Selected_TextBlock_Current_TournamentHealingPerk : text_block = text_block{DefaultTextColor := CurrentTextColor, DefaultShadowOffset:=ShadowOffset, DefaultShadowOpacity := DefaultTextShadowOpacity}


    #PLAYER STATS
        #Package A
    var GeneralSelector_Selected_TextBlock_Description_SpecialPerkA : text_block = text_block{DefaultTextColor := Pink, DefaultShadowOffset:=ShadowOffset, DefaultShadowOpacity := DefaultTextShadowOpacity}

    var GeneralSelector_Selected_TextBlock_Current_SpecialPerkA : text_block = text_block{DefaultTextColor := CurrentTextColor, DefaultShadowOffset:=ShadowOffset, DefaultShadowOpacity := DefaultTextShadowOpacity}
    var GeneralSelector_Selected_TextBlock_Current_PlayerPickDamage : text_block = text_block{DefaultTextColor := CurrentTextColor, DefaultShadowOffset:=ShadowOffset, DefaultShadowOpacity := DefaultTextShadowOpacity}
    var GeneralSelector_Selected_TextBlock_Current_PlayerWeakPoint : text_block = text_block{DefaultTextColor := CurrentTextColor, DefaultShadowOffset:=ShadowOffset, DefaultShadowOpacity := DefaultTextShadowOpacity}

        #Package B
    var GeneralSelector_Selected_TextBlock_Description_SpecialPerkB : text_block = text_block{DefaultTextColor := Pink, DefaultShadowOffset:=ShadowOffset, DefaultShadowOpacity := DefaultTextShadowOpacity}

    var GeneralSelector_Selected_TextBlock_Current_SpecialPerkB : text_block = text_block{DefaultTextColor := CurrentTextColor, DefaultShadowOffset:=ShadowOffset, DefaultShadowOpacity := DefaultTextShadowOpacity}
    var GeneralSelector_Selected_TextBlock_Current_PlayerPickEfficiency : text_block = text_block{DefaultTextColor := CurrentTextColor, DefaultShadowOffset:=ShadowOffset, DefaultShadowOpacity := DefaultTextShadowOpacity}
    var GeneralSelector_Selected_TextBlock_Current_PlayerPickCriticalDamage : text_block = text_block{DefaultTextColor := CurrentTextColor, DefaultShadowOffset:=ShadowOffset, DefaultShadowOpacity := DefaultTextShadowOpacity}

    OnPlayerRemoved() : void =
        for(Prop : AllPetsFollowProp):
            Prop.Dispose()

    Subscribes() : void =
        #PrintWarning("On pet_manager CheckPetAction must be checked by the Game Manager in the future")
        #PrintWarning("On pet_manager PlayerEnhacement_MobHunter need to be used on the pick action")
        #Perks
        BuySlot_ActiveSlots_Button.OnClick().Subscribe(BuySlots_TryImproveActiveSlots)
        BuySlot_MaxSlots_Button.OnClick().Subscribe(BuySlots_TryImproveMaxSlots)
        BuySlot_TierUnlock_Button.OnClick().Subscribe(BuySlots_TryImproveMaxPetTier)
    
        BuySlot_Close_Button.OnClick().Subscribe(handler_hexgame{MaybePetsManager:=option{Self}, SubMenuOption:=SUBMENU_BUY_SLOTSPET}.HandlerUIFromEverythingToMainMenu)

        #Recycle pet
        RecyclePet_ExecuteRecycleButton.OnClick().Subscribe(HandlerExecuteRecycle)
        RecyclePet_NavigateRightButton.OnClick().Subscribe(RecyclePet_NavidateRight)
        RecyclePet_NavigateLeftButton.OnClick().Subscribe(RecyclePet_NavidateLeft)
        RecyclePet_CancelButton.OnClick().Subscribe(handler_hexgame{MaybePetsManager:=option{Self}, SubMenuOption:=SUBMENU_RECYCLE_PET}.HandlerUIFromEverythingToMainMenu)

        GeneralSelector_SelectButton.OnClick().Subscribe(HandlerGeneralSelector_SelectPet)
        #Equip pet
        EquipPet_0_ChangeSlotButton.OnClick().Subscribe(handler_hexgame{MaybePetsManager:=option{Self}, SubMenuOption:=SUBMENU_EQUIP_PET, UISlot := 0}.HandlerUIGeneralSelector)
        EquipPet_1_ChangeSlotButton.OnClick().Subscribe(handler_hexgame{MaybePetsManager:=option{Self}, SubMenuOption:=SUBMENU_EQUIP_PET, UISlot := 1}.HandlerUIGeneralSelector)
        EquipPet_2_ChangeSlotButton.OnClick().Subscribe(handler_hexgame{MaybePetsManager:=option{Self}, SubMenuOption:=SUBMENU_EQUIP_PET, UISlot := 2}.HandlerUIGeneralSelector)
        EquipPet_CloseButton.OnClick().Subscribe(handler_hexgame{MaybePetsManager:=option{Self}, SubMenuOption:=SUBMENU_EQUIP_PET}.HandlerUIFromEverythingToMainMenu)

        #Create Pet
        CreatePet_CreatePetButton.OnClick().Subscribe(CreatePet_TryCreatePet)

        CreatePet_CloseButton.OnClick().Subscribe(handler_hexgame{MaybePetsManager:=option{Self}, SubMenuOption:=SUBMENU_CREATE_PET}.HandlerUIFromEverythingToMainMenu)
        CreatePet_ChangeClassButton.OnClick().Subscribe(CreatePet_ChangeClass)

        CreatePet_Add1PurityButton.OnClick().Subscribe(handler_hexgame{MaybePetsManager:=option{Self}, PurityAmountChange:=1}.HandlerUICreatePet_AddOrRemovePurity)
        CreatePet_Add5PurityButton.OnClick().Subscribe(handler_hexgame{MaybePetsManager:=option{Self}, PurityAmountChange:=5}.HandlerUICreatePet_AddOrRemovePurity)
        CreatePet_Add25PurityButton.OnClick().Subscribe(handler_hexgame{MaybePetsManager:=option{Self}, PurityAmountChange:=25}.HandlerUICreatePet_AddOrRemovePurity)

        CreatePet_Remove1PurityButton.OnClick().Subscribe(handler_hexgame{MaybePetsManager:=option{Self}, PurityAmountChange:=-1}.HandlerUICreatePet_AddOrRemovePurity)
        CreatePet_Remove5PurityButton.OnClick().Subscribe(handler_hexgame{MaybePetsManager:=option{Self}, PurityAmountChange:=-5}.HandlerUICreatePet_AddOrRemovePurity)
        CreatePet_Remove25PurityButton.OnClick().Subscribe(handler_hexgame{MaybePetsManager:=option{Self}, PurityAmountChange:=-25}.HandlerUICreatePet_AddOrRemovePurity)

        #Main menu
        MainMenu_CreatePetButton.OnClick().Subscribe(handler_hexgame{MaybePetsManager:=option{Self}, SubMenuOption:=SUBMENU_CREATE_PET}.HandlerUIPetSubmenu)
        MainMenu_EquipPetButton.OnClick().Subscribe((handler_hexgame{MaybePetsManager:=option{Self}, SubMenuOption:=SUBMENU_EQUIP_PET}.HandlerUIPetSubmenu))
        MainMenu_ImprovePetButton.OnClick().Subscribe((handler_hexgame{MaybePetsManager:=option{Self}, SubMenuOption:=SUBMENU_IMPROVE_PET}.HandlerUIPetSubmenu))
        MainMenu_RecyclePetButton.OnClick().Subscribe((handler_hexgame{MaybePetsManager:=option{Self}, SubMenuOption:=SUBMENU_RECYCLE_PET}.HandlerUIPetSubmenu))
        MainMenu_BuySlotsPetButton.OnClick().Subscribe((handler_hexgame{MaybePetsManager:=option{Self}, SubMenuOption:=SUBMENU_BUY_SLOTSPET}.HandlerUIPetSubmenu))
        MainMenu_MixPetsButton.OnClick().Subscribe((handler_hexgame{MaybePetsManager:=option{Self}, SubMenuOption:=SUBMENU_MIX_PET}.HandlerUIPetSubmenu))
        MainMenu_CloseButton.OnClick().Subscribe(CloseAllMainMenu)

        #PetCreation_CreatePetMenuButton.OnClick().Subscribe(CreationMainMenu_TryCreatePet)
        #CreationMainMenu_MixPetMenuButton.OnClick().Subscribe(Handler_MainMenu_To_MixPetsMenu)

        MixPets_CloseButton.OnClick().Subscribe(handler_hexgame{MaybePetsManager:=option{Self}, SubMenuOption:=SUBMENU_MIX_PET}.HandlerUIFromEverythingToMainMenu)
        
        MixPets_Button_Box0.OnClick().Subscribe(handler_hexgame{MaybePetsManager:=option{Self},SubMenuOption:=SUBMENU_MIX_PET,UISlot:=0}.HandlerUIGeneralSelector)
        MixPets_Button_Box1.OnClick().Subscribe(handler_hexgame{MaybePetsManager:=option{Self},SubMenuOption:=SUBMENU_MIX_PET,UISlot:=1}.HandlerUIGeneralSelector)
        MixPets_MixButton.OnClick().Subscribe(MixPets_TryToMixPets)

        #GeneralSelector_SelectButton.OnClick().Subscribe(handler_hexgame{MaybePetsManager:=option{Self},UIComesFrom:=From_MixPetsUI,UISlot:=1}.HandlerUIPetGeneralSelector)

        GeneralSelector_NavigateRightButton.OnClick().Subscribe(GeneralSelector_NavidateRight)
        GeneralSelector_NavigateLeftButton.OnClick().Subscribe(GeneralSelector_NavidateLeft)
        GeneralSelector_CancelButton.OnClick().Subscribe(GeneralSelector_ExitPetSelector)

        #Recicle Pet
        RecyclePet_NavigateRightButton.OnClick().Subscribe(GeneralSelector_NavidateRight)
        RecyclePet_NavigateLeftButton.OnClick().Subscribe(GeneralSelector_NavidateLeft)
        RecyclePet_CancelButton.OnClick().Subscribe(handler_hexgame{MaybePetsManager:=option{Self}, SubMenuOption:=SUBMENU_RECYCLE_PET}.HandlerUIFromEverythingToMainMenu)

    MovePet() : void =
        #PrintWarning("On pet_manager MovePet must be checked by the Game Manager in the future")
            for(Index:=0..AllPetsFollowProp.Length-1, Prop:=AllPetsFollowProp[Index], Prop.IsValid[], Tier := AllPetsFollowProp_Tier[Index]):
                if(Agent:=MaybeAgent?):
                    spawn:
                        AnimatePetMovement(Agent, Prop, Tier, Index)

    Generate_PlayerEnhacements() : void =
        if(Agent:=MaybeAgent?):
            MaybeBakery:=GameManager.GetAgentBakery(Agent)
            if(ValidBakery := MaybeBakery?):
                if(Player := player[Agent], Player.IsActive[]):
                    EquippedPetsTupple := GetEquippedPets(ValidBakery.PlayerDataCache)
                    set AllEquippedPetsIndexes=EquippedPetsTupple(1)
                    set AllEquippedPetTables=EquippedPetsTupple(0)
                    loop:
                        if(AllEquippedPetsIndexes.Length<3):
                            set AllEquippedPetsIndexes = AllEquippedPetsIndexes + array{-1}
                        else:
                            break
        #Reset player stats
            #COMMON
        set PlayerEnhacement_PickDamage = 1.0
        set PlayerEnhacement_WeakPointMult = 1.0
        set PlayerEnhacement_PickEfficency = 1.0
        set PlayerEnhacement_PickCriticalDam = 1.0
            #ADC
        set PlayerEnhacement_BowExtraDamage = 1.0
        set PlayerEnhacement_ArrowCraft = 0.0
            #DRUID
        set PlayerEnhacement_HealthRegeneration = 0.0
        set PlayerEnhacement_EmergencyHealing = 0.0
            #TANK
        set PlayerEnhacement_ExtraHealth = 0.0
        set PlayerEnhacement_EmergencyShield  = 0.0
            #FIGHTER
        set PlayerEnhacement_SwordExtraDamage = 1.0
        set PlayerEnhacement_SwordSuperStrike = 1.0
        set PlayerEnhacement_SwordSuperStrikeNum = 0.0
            #HUNTER
        set PlayerEnhacement_MobHunter = 1.0
        set PlayerEnhacement_WeaponsExtraCritical = 1.0
        #Prepare acumulations
            #COMMON
        var Accumulated_PlayerEnhacement_PickDamage : float = 1.0
        var Accumulated_PlayerEnhacement_WeakPointMult : float = 1.0
        var Accumulated_PlayerEnhacement_PickEfficency : float = 1.0
        var Accumulated_PlayerEnhacement_PickCriticalDam : float = 1.0
            #ADC
        var Accumulated_PlayerEnhacement_BowExtraDamage : float = 1.0
        var Accumulated_PlayerEnhacement_ArrowCraft : float = 0.0
            #DRUID
        var Accumulated_PlayerEnhacement_HealthRegeneration : float = 0.0
        var Accumulated_PlayerEnhacement_EmergencyHealing : float = 0.0
            #TANK
        var Accumulated_PlayerEnhacement_ExtraHealth : float = 0.0
        var Accumulated_PlayerEnhacement_EmergencyShield : float = 0.0
            #FIGHTER
        var Accumulated_PlayerEnhacement_SwordExtraDamage : float = 1.0
        var Accumulated_PlayerEnhacement_SwordSuperStrike : float = 1.0
        var Accumulated_PlayerEnhacement_SwordSuperStrikeNum : float = 0.0
            #HUNTER
        var Accumulated_PlayerEnhacement_MobHunter : float = 1.0
        var Accumulated_PlayerEnhacement_WeaponsExtraCritical : float = 1.0

        for(Index:=0..AllEquippedPetTables.Length-1, PetTable:=AllEquippedPetTables[Index]):
            if(Agent:=MaybeAgent?, Character := Agent.GetFortCharacter[]):
                #Common
                set Accumulated_PlayerEnhacement_PickDamage = Accumulated_PlayerEnhacement_PickDamage + (MainFormula_EnhancementsPetStats(STAT_TYPE_PLAYER_PICKDAMAGE, PetTable.Type, PetTable.LVPet_PackageA, PetTable.Purity)-1.0)
                set Accumulated_PlayerEnhacement_WeakPointMult = Accumulated_PlayerEnhacement_WeakPointMult + (MainFormula_EnhancementsPetStats(STAT_TYPE_PLAYER_WEAKPOINTMULT, PetTable.Type, PetTable.LVPet_PackageA, PetTable.Purity)-1.0)
                set Accumulated_PlayerEnhacement_PickEfficency = Accumulated_PlayerEnhacement_PickEfficency + (MainFormula_EnhancementsPetStats(STAT_TYPE_PLAYER_PICKEFFICENCY, PetTable.Type, PetTable.LVPet_PackageB, PetTable.Purity)-1.0)
                set Accumulated_PlayerEnhacement_PickCriticalDam = Accumulated_PlayerEnhacement_PickCriticalDam + (MainFormula_EnhancementsPetStats(STAT_TYPE_PLAYER_PICKCRITICALMULT, PetTable.Type, PetTable.LVPet_PackageB, PetTable.Purity)-1.0)

                #Special amounts
                SpecialA_Ammount:=MainFormula_EnhancementsPetStats(STAT_TYPE_PLAYER_SPECIAL_A, PetTable.Type, PetTable.LVPet_PackageA, PetTable.Purity)
                SpecialB_Ammount:=MainFormula_EnhancementsPetStats(STAT_TYPE_PLAYER_SPECIAL_B, PetTable.Type, PetTable.LVPet_PackageB, PetTable.Purity)
                if(PetTable.Type=ADC):
                    set Accumulated_PlayerEnhacement_BowExtraDamage = Accumulated_PlayerEnhacement_BowExtraDamage + (SpecialA_Ammount-1.0)
                    set Accumulated_PlayerEnhacement_ArrowCraft = Accumulated_PlayerEnhacement_ArrowCraft + (100.0/SpecialB_Ammount)
                if(PetTable.Type=DRUID):
                    set Accumulated_PlayerEnhacement_HealthRegeneration = Accumulated_PlayerEnhacement_HealthRegeneration + SpecialA_Ammount
                    set Accumulated_PlayerEnhacement_EmergencyHealing = Accumulated_PlayerEnhacement_EmergencyHealing + SpecialB_Ammount
                if(PetTable.Type=TANK):
                    set Accumulated_PlayerEnhacement_ExtraHealth = Accumulated_PlayerEnhacement_ExtraHealth + SpecialA_Ammount
                    set Accumulated_PlayerEnhacement_EmergencyShield = Accumulated_PlayerEnhacement_EmergencyShield + SpecialB_Ammount
                if(PetTable.Type=FIGHTER):
                    set Accumulated_PlayerEnhacement_SwordExtraDamage = Accumulated_PlayerEnhacement_SwordExtraDamage + (SpecialA_Ammount-1.0)
                    set Accumulated_PlayerEnhacement_SwordSuperStrike = Accumulated_PlayerEnhacement_SwordSuperStrike + (SpecialB_Ammount-1.0) 
                if(PetTable.Type=HUNTER):
                    set Accumulated_PlayerEnhacement_MobHunter = Accumulated_PlayerEnhacement_MobHunter + (SpecialA_Ammount-1.0)
                    set Accumulated_PlayerEnhacement_WeaponsExtraCritical = Accumulated_PlayerEnhacement_WeaponsExtraCritical + (SpecialB_Ammount-1.0)
            #COMMON
        set PlayerEnhacement_PickDamage = Accumulated_PlayerEnhacement_PickDamage
        set PlayerEnhacement_WeakPointMult = Accumulated_PlayerEnhacement_WeakPointMult
        set PlayerEnhacement_PickEfficency = Accumulated_PlayerEnhacement_PickEfficency
        set PlayerEnhacement_PickCriticalDam = Accumulated_PlayerEnhacement_PickCriticalDam
            #ADC
        set PlayerEnhacement_BowExtraDamage = Accumulated_PlayerEnhacement_BowExtraDamage
        set PlayerEnhacement_ArrowCraft = Accumulated_PlayerEnhacement_ArrowCraft
            #DRUID
        set PlayerEnhacement_HealthRegeneration = Accumulated_PlayerEnhacement_HealthRegeneration
        set PlayerEnhacement_EmergencyHealing = Accumulated_PlayerEnhacement_EmergencyHealing
        ##Print("PlayerEnhacement_EmergencyHealing is {PlayerEnhacement_EmergencyHealing}")
            #TANK
        set PlayerEnhacement_ExtraHealth = Accumulated_PlayerEnhacement_ExtraHealth
        set PlayerEnhacement_EmergencyShield = Accumulated_PlayerEnhacement_EmergencyShield
            #FIGHTER
        set PlayerEnhacement_SwordExtraDamage = Accumulated_PlayerEnhacement_SwordExtraDamage
        set PlayerEnhacement_SwordSuperStrike = Accumulated_PlayerEnhacement_SwordSuperStrike
            #HUNTER
        set PlayerEnhacement_MobHunter = Accumulated_PlayerEnhacement_MobHunter
        set PlayerEnhacement_WeaponsExtraCritical = Accumulated_PlayerEnhacement_WeaponsExtraCritical

        if(Agent:=MaybeAgent?, Character := Agent.GetFortCharacter[]):
            Character.SetMaxHealth(PlayerBaseHealth+PlayerEnhacement_ExtraHealth)
            Character.SetHealth(PlayerBaseHealth+PlayerEnhacement_ExtraHealth)
            Character.SetMaxShield(PlayerEnhacement_EmergencyShield)
            Log("Generated Player MaxHealth: {PlayerBaseHealth+PlayerEnhacement_ExtraHealth}",LOG_LEVEL_DEBUG6)
            spawn:
                CheckPetAction(GameManager.TotalGameTime,false)

    CheckPetAction(TotalGameTime : float, GameManagerCheck : logic)<suspends> : void =
        var AllPerkProgress : []float = array{}
        var AllPetsType : []int = array{}
        if(Agent := MaybeAgent?, Character := Agent.GetFortCharacter[]):
            ExecuteShieldDegradation(Character)
            #ADC
            if(PlayerEnhacement_ArrowCraft>0.0):
                if(ArrowCraftCreation_Progress >= 100.0):
                    GameManager.GeneralMobsManager.WeaponsGranter.SetNextItem(4)
                    GameManager.GeneralMobsManager.WeaponsGranter.GrantItem(Agent)
                    GameManager.VFXManager.ArrowGranted_AudioPlayer.Play(Agent)
                    set ArrowCraftCreation_Progress = 0.0
                    if(ValidBakery := GameManager.GetAgentBakery_Transacted(Agent)?, MainInventoryUI := ValidBakery.MaybeMainInventoryUI?, ActionUI := ValidBakery.MaybeActionInfoUI?):
                        ActionUI.DisplayGather(array{1.0}, array{ARROW}, array{false},false)
                var PerkProgress : float = ArrowCraftCreation_Progress/100.0
                if(PerkProgress>1.0):
                    set PerkProgress = 1.0
                set AllPerkProgress = AllPerkProgress + array{PerkProgress}
                set AllPetsType = AllPetsType + array{ADC}
                if(GameManagerCheck?):
                    set ArrowCraftCreation_Progress = ArrowCraftCreation_Progress + PlayerEnhacement_ArrowCraft
            #DRUID
            Character.Heal(PlayerEnhacement_HealthRegeneration)
            if(PlayerEnhacement_EmergencyHealing>0.0): 
                if(Character.GetHealth()<Character.GetMaxHealth(), TotalGameTime > PlayerEnhacement_HealLastTimeUsed+PetPlayerEnhacement_HealDelay):
                    #Print("Executing emergecy heal")
                    set PlayerEnhacement_HealLastTimeUsed = TotalGameTime
                    Character.Heal(PlayerEnhacement_EmergencyHealing)
                    spawn:
                        GameManager.VFXManager.SpawnHealEffect(false,Agent)
                #else:
                    ##Print("Recharguing heal: {((TotalGameTime-PlayerEnhacement_HealLastTimeUsed)*1.0)/(PetPlayerEnhacement_HealDelay*1.0)}")
                var PerkProgress : float = ((TotalGameTime-PlayerEnhacement_HealLastTimeUsed)*1.0)/(PetPlayerEnhacement_HealDelay*1.0)
                if(PerkProgress>1.0):
                    set PerkProgress = 1.0
                set AllPerkProgress = AllPerkProgress + array{PerkProgress}
                set AllPetsType = AllPetsType + array{DRUID}
            #TANK
            if(PlayerEnhacement_EmergencyShield>0.0):
                if(Character.GetHealth()<(Character.GetMaxHealth()*0.75), TotalGameTime > PlayerEnhacement_ShieldLastTimeUsed+PetPlayerEnhacement_ShieldDelay):
                    #Print("Executing emergeny shield")
                    spawn:
                        GameManager.VFXManager.SpawnHealEffect(true,Agent)
                    set PlayerEnhacement_ShieldLastTimeUsed = TotalGameTime
                    MaxShield:=Character.GetMaxShield()
                    NewShieldAmount:=Character.GetShield()+PlayerEnhacement_EmergencyShield
                    ##Print("NewShieldAmount: {NewShieldAmount}")
                    if(NewShieldAmount>MaxShield):
                        Character.SetShield(MaxShield)
                    else:
                        Character.SetShield(NewShieldAmount)
                var PerkProgress : float = ((TotalGameTime-PlayerEnhacement_ShieldLastTimeUsed)*1.0)/(PetPlayerEnhacement_ShieldDelay*1.0)
                if(PerkProgress>1.0):
                    set PerkProgress = 1.0
                set AllPerkProgress = AllPerkProgress + array{PerkProgress}
                set AllPetsType = AllPetsType + array{TANK}
            #FIGHTER
            if(PlayerEnhacement_SwordSuperStrike>1.0):
                Progress:=PlayerEnhacement_SwordSuperStrikeNum/NUMHITS_FOR_SUPERSTRIKE
                set AllPerkProgress = AllPerkProgress + array{Progress}
                set AllPetsType = AllPetsType + array{FIGHTER}
                if(Progress>=1.0):
                    GameManager.GeneralMobsManager.SuperStrikeEffect.Pickup(Agent)
                else:
                    GameManager.GeneralMobsManager.RemoveSuperStrikeEffect.Pickup(Agent)
            #HUNTER
            if(PlayerEnhacement_MobHunter>1.0):
                set AllPerkProgress = AllPerkProgress + array{1.0}
                set AllPetsType = AllPetsType + array{HUNTER}
            MaybeBakery := GameManager.GetAgentBakery(Agent)
            if(ValidBakery:=MaybeBakery?, MainInventoryUI:=ValidBakery.MaybeMainInventoryUI?):
                MainInventoryUI.RefreshPetPerk_Sliders(AllPerkProgress,AllPetsType)   

    ExecuteShieldDegradation(Character : fort_character) : void =
        CurrentShield:=Character.GetShield()
        if(CurrentShield>0.0):
            Character.SetShield(CurrentShield-1.0)

    SetAgent(InAgent : agent) : void =
        GeneralSelector_Remove_CanvasWithoutMessage()
        MainMenu_Remove_CanvasWithoutMessage()
        MixPets_Remove_CanvasWithoutMessage()
        EquipPet_Remove_CanvasWithoutMessage()
        RecyclePet_Remove_CanvasWithoutMessage()
        CreatePet_Remove_CanvasWithoutMessage()
        BuySlots_Remove_CanvasWithoutMessage()

        #UI FOCUS SECURITY
        set GeneralSelector_UIIsDisplayed = false
        set MainMenu_UIIsDisplayed = false
        set MixPets_UIIsDisplayed = false
        set EquipPet_UIIsDisplayed = false
        set RecyclePet_UIIsDisplayed = false
        set CreatePet_UIIsDisplayed = false
        set BuySlots_UIIsDisplayed = false

        set AllEquippedPetsIndexes = array{-1,-1,-1}
        set FusionedPet = pet_data_table{}
        set IsFusionReady = false
        set AllMixPets_PetsToMix = array{-1,-1}

        set ConfirmationToRecycle = 0
        set ConfirmationToMix = false
        set ConfirmationToCreate = false
        if(Agent:=MaybeAgent?):
            MaybeBakery:=GameManager.GetAgentBakery(Agent)
            if(ValidBakery := MaybeBakery?):
                if(Player := player[InAgent], Player.IsActive[]):
                    EquippedPetsTupple := GetEquippedPets(ValidBakery.PlayerDataCache)
                    set AllEquippedPetsIndexes=EquippedPetsTupple(1)
                    set AllEquippedPetTables=EquippedPetsTupple(0)
                    loop:
                        if(AllEquippedPetsIndexes.Length<3):
                            set AllEquippedPetsIndexes = AllEquippedPetsIndexes + array{-1}
                        else:
                            break
                    LoadEquippedPets(ValidBakery.PlayerDataCache)

        set MaybeAgent = option{InAgent}
        if(PetsUpgradesUI:=MaybePetsUpgradesUI?):
            PetsUpgradesUI.Remove_CanvasWithoutMessage()
            PetsUpgradesUI.SetAgent(InAgent)

        Generate_PlayerEnhacements()

    CloseAllUI() : void =
        if(PetsUpgradesUI:=MaybePetsUpgradesUI?):
            PetsUpgradesUI.Remove_CanvasWithoutMessage()
        GeneralSelector_Remove_CanvasWithoutMessage()
        MainMenu_Remove_CanvasWithoutMessage()
        MixPets_Remove_CanvasWithoutMessage()
        EquipPet_Remove_CanvasWithoutMessage()
        RecyclePet_Remove_CanvasWithoutMessage()
        CreatePet_Remove_CanvasWithoutMessage()
        BuySlots_Remove_CanvasWithoutMessage()


    CheckSwordStrikeEvent(TotalGameTime : float, Amount : float) : tuple(float,logic) =
        if(PlayerEnhacement_SwordSuperStrikeNum>=NUMHITS_FOR_SUPERSTRIKE):
            if(LastHammerStrike+0.1<TotalGameTime):
                set LastHammerStrike = TotalGameTime
                set PlayerEnhacement_SwordSuperStrikeNum = 0.0
            #CheckPetAction(TotalgameTime)
            if(Agent:=MaybeAgent?, Character := Agent.GetFortCharacter[]):
                Character.SetMaxHealth(PlayerBaseHealth+PlayerEnhacement_ExtraHealth)
                spawn:
                    CheckPetAction(GameManager.TotalGameTime,false)
            return (Amount*PlayerEnhacement_SwordSuperStrike,true)
        else:
            if(PlayerEnhacement_SwordSuperStrike>1.0, Agent:=MaybeAgent?):
                if(LastHammerStrike+0.1<TotalGameTime):
                    set LastHammerStrike = TotalGameTime
                    set PlayerEnhacement_SwordSuperStrikeNum = PlayerEnhacement_SwordSuperStrikeNum + 1.0
                    if(PlayerEnhacement_SwordSuperStrikeNum=NUMHITS_FOR_SUPERSTRIKE):
                        GameManager.VFXManager.SuperStrikeLoaded_AudioPlayer.Play(Agent)
            if(Agent:=MaybeAgent?, Character := Agent.GetFortCharacter[]):
                Character.SetMaxHealth(PlayerBaseHealth+PlayerEnhacement_ExtraHealth)
                spawn:
                    CheckPetAction(GameManager.TotalGameTime,false)
            return (Amount,false)


    GetPlayerEnhancements(PlayerData : player_data_table) : tuple(float,float,float,float,float,float,float,float,float,float,float,float,float,float) =

        var TotalBowDamage_Multiplier : float = 1.0
        var TotalArrowProduction_LessDelay : float = 0.0

        var TotalPasiveHeal : float = 0.0
        var TotalEmergencyHeal : float = 0.0

        var TotalExtraHealth : float = 0.0
        var TotalCombatShield : float = 0.0

        var TotalSwordDamage_Multiplier : float = 1.0
        var TotalSwordUltraStrike_Multiplier : float = 1.0

        var TotalHuntProbability_Multiplier : float = 1.0
        var TotalWeaponCritical_Multiplier : float = 1.0

        var TotalPickDamage_Multiplier : float = 1.0
        var TotalWeakPoint_Multiplier : float = 1.0
        var TotalPickEfficency_Multiplier : float = 1.0
        var TotalPickCriticalDamage_Multiplier : float = 1.0

        for(PetIndex : AllEquippedPetsIndexes, Pet := PlayerData.AllPets[PetIndex]):
            #Package A
            set TotalPickDamage_Multiplier = TotalPickDamage_Multiplier + (MainFormula_EnhancementsPetStats(STAT_TYPE_PLAYER_PICKDAMAGE,Pet.Type,Pet.LVPet_PackageA,Pet.Purity)-1.0)
            set TotalWeaponCritical_Multiplier = TotalWeaponCritical_Multiplier + (MainFormula_EnhancementsPetStats(STAT_TYPE_PLAYER_WEAKPOINTMULT,Pet.Type,Pet.LVPet_PackageA,Pet.Purity)-1.0)
            #Package B
            set TotalPickEfficency_Multiplier = TotalPickEfficency_Multiplier + (MainFormula_EnhancementsPetStats(STAT_TYPE_PLAYER_PICKEFFICENCY,Pet.Type,Pet.LVPet_PackageB,Pet.Purity)-1.0)
            set TotalPickCriticalDamage_Multiplier = TotalPickCriticalDamage_Multiplier + (MainFormula_EnhancementsPetStats(STAT_TYPE_PLAYER_PICKCRITICALMULT,Pet.Type,Pet.LVPet_PackageB,Pet.Purity)-1.0)
            #Special package A
            if(Pet.Type = ADC):
                set TotalBowDamage_Multiplier = TotalBowDamage_Multiplier + (MainFormula_EnhancementsPetStats(STAT_TYPE_PLAYER_SPECIAL_A,Pet.Type,Pet.LVPet_PackageA,Pet.Purity)-1.0)
            else if(Pet.Type = DRUID):
                set TotalPasiveHeal = TotalPasiveHeal + MainFormula_EnhancementsPetStats(STAT_TYPE_PLAYER_SPECIAL_A,Pet.Type,Pet.LVPet_PackageA,Pet.Purity)
            else if(Pet.Type = TANK):
                set TotalExtraHealth = TotalExtraHealth + MainFormula_EnhancementsPetStats(STAT_TYPE_PLAYER_SPECIAL_A,Pet.Type,Pet.LVPet_PackageA,Pet.Purity)
            else if(Pet.Type = FIGHTER):
                set TotalExtraHealth = TotalExtraHealth + (MainFormula_EnhancementsPetStats(STAT_TYPE_PLAYER_SPECIAL_A,Pet.Type,Pet.LVPet_PackageA,Pet.Purity)-1.0)
            else if(Pet.Type = HUNTER): 
                set TotalHuntProbability_Multiplier = TotalHuntProbability_Multiplier + (MainFormula_EnhancementsPetStats(STAT_TYPE_PLAYER_SPECIAL_A,Pet.Type,Pet.LVPet_PackageA,Pet.Purity)-1.0)
            #SPECIAL PACKAGE B
            if(Pet.Type = ADC):
                AmountLess:=ADC_PLAYER_SPECIALB_BASETIME - MainFormula_EnhancementsPetStats(STAT_TYPE_PLAYER_SPECIAL_B,Pet.Type,Pet.LVPet_PackageA,Pet.Purity)
                if(AmountLess<ADC_PLAYER_SPECIALB_BASETIME):
                    set TotalArrowProduction_LessDelay = TotalArrowProduction_LessDelay + AmountLess
            else if(Pet.Type = DRUID):
                set TotalEmergencyHeal = TotalEmergencyHeal + MainFormula_EnhancementsPetStats(STAT_TYPE_PLAYER_SPECIAL_B,Pet.Type,Pet.LVPet_PackageA,Pet.Purity)
            else if(Pet.Type = TANK):
                set TotalCombatShield = TotalCombatShield + MainFormula_EnhancementsPetStats(STAT_TYPE_PLAYER_SPECIAL_B,Pet.Type,Pet.LVPet_PackageA,Pet.Purity)
            else if(Pet.Type = FIGHTER):
                set TotalSwordUltraStrike_Multiplier = TotalSwordUltraStrike_Multiplier + (MainFormula_EnhancementsPetStats(STAT_TYPE_PLAYER_SPECIAL_B,Pet.Type,Pet.LVPet_PackageA,Pet.Purity)-1.0)
            else if(Pet.Type = HUNTER): 
                set TotalWeaponCritical_Multiplier = TotalWeaponCritical_Multiplier + (MainFormula_EnhancementsPetStats(STAT_TYPE_PLAYER_SPECIAL_B,Pet.Type,Pet.LVPet_PackageA,Pet.Purity)-1.0)

        return (TotalBowDamage_Multiplier,TotalArrowProduction_LessDelay,
            TotalPasiveHeal,TotalEmergencyHeal,TotalExtraHealth,TotalCombatShield,TotalSwordDamage_Multiplier,
            TotalSwordUltraStrike_Multiplier,TotalHuntProbability_Multiplier,TotalWeaponCritical_Multiplier,
            TotalPickDamage_Multiplier,TotalWeakPoint_Multiplier,TotalPickEfficency_Multiplier,TotalPickCriticalDamage_Multiplier)


    CreateUpgradesUI() : void =
        PetsUpgradesUI:=MakePetUpgradesUI(Self)
        set MaybePetsUpgradesUI = option{PetsUpgradesUI}
        PetsUpgradesUI.InitUI()

    InitUIs() : void =

        #MIX PETS
        RecyclePet_NavigateRightButton.SetText(BasicMessage("▶"))
        RecyclePet_NavigateLeftButton.SetText(BasicMessage("◀"))

        MixPets_MixTitle.SetText(BasicMessage("SOURDOUGH MIXTURE OUTPUT"))
        MixPets_MixButton.SetText(BasicMessage("MIX"))
        MixPets_Button_Box0.SetText(BasicMessage("SELECT"))
        MixPets_Button_Box1.SetText(BasicMessage("SELECT"))
        MixPets_CloseButton.SetText(BasicMessage("CLOSE "))

        #CREATE PET
        #PrintWarning("TODO: Only the owner of the bakery should open the pet manager, and reset player when this leaves, so the pet is not tring to move")

        CreatePet_Add1PurityButton.SetText(BasicMessage("+1 PURITY"))
        CreatePet_Add5PurityButton.SetText(BasicMessage("+5 PURITY"))
        CreatePet_Add25PurityButton.SetText(BasicMessage("+25 PURITY"))

        CreatePet_Remove1PurityButton.SetText(BasicMessage("-1 PURITY"))
        CreatePet_Remove5PurityButton.SetText(BasicMessage("-5 PURITY"))
        CreatePet_Remove25PurityButton.SetText(BasicMessage("-25 PURITY"))

        CreatePet_ChangeClassButton.SetText(BasicMessage("CHANGE CLASS"))

        GeneralSelector_NavigateRightButton.SetText(BasicMessage("▶"))
        GeneralSelector_NavigateLeftButton.SetText(BasicMessage("◀"))
        GeneralSelector_CancelButton.SetText(BasicMessage("Cancel"))
        GeneralSelector_SelectButton.SetText(BasicMessage("Select"))

        MainMenu_CreatePetButton.SetText(BasicMessage("Create new Sourdough"))
        MainMenu_EquipPetButton.SetText(BasicMessage("Change sourdough slots"))
        MainMenu_ImprovePetButton.SetText(BasicMessage("Improve Sourdough"))
        MainMenu_RecyclePetButton.SetText(BasicMessage("Recycle Sourdough"))
        MainMenu_BuySlotsPetButton.SetText(BasicMessage("Buy slots / tiers"))
        MainMenu_MixPetsButton.SetText(BasicMessage("Mix sourdoughs"))
        MainMenu_CloseButton.SetText(BasicMessage("Close "))

        #Equip
        #All
        set AllEquipPet_SlotInfo = array{EquipPet_0_SlotInfo,EquipPet_1_SlotInfo,EquipPet_2_SlotInfo}
        set AllEquipPet_ChangeSlot = array{EquipPet_0_ChangeSlotButton,EquipPet_1_ChangeSlotButton,EquipPet_2_ChangeSlotButton}
        set AllEquipPet_Text_PackageA_LV = array{EquipPet_0_Text_PackageA_LV,EquipPet_1_Text_PackageA_LV,EquipPet_2_Text_PackageA_LV}
        set AllEquipPet_Text_PackageB_LV = array{EquipPet_0_Text_PackageB_LV,EquipPet_1_Text_PackageB_LV,EquipPet_2_Text_PackageB_LV}
        #TOURNAMENT
        set AllEquipPet_TextureBlock_PetImages = array{EquipPet_0_TextureBlock_PetImages,EquipPet_1_TextureBlock_PetImages,EquipPet_2_TextureBlock_PetImages}
        set AllEquipPet_TextBlock_Purity = array{EquipPet_0_TextBlock_Purity,EquipPet_1_TextBlock_Purity,EquipPet_2_TextBlock_Purity}
        set AllEquipPet_TextBlock_Class = array{EquipPet_0_TextBlock_Class,EquipPet_1_TextBlock_Class,EquipPet_2_TextBlock_Class}
        set AllEquipPet_TextBlock_PackA = array{EquipPet_0_TextBlock_PackA,EquipPet_1_TextBlock_PackA,EquipPet_2_TextBlock_PackA}
        set AllEquipPet_TextBlock_PackB = array{EquipPet_0_TextBlock_PackB,EquipPet_1_TextBlock_PackB,EquipPet_2_TextBlock_PackB}
        #Package A
        set AllEquipPet_TextBlock_Description_SpecialPerkA = array{EquipPet_0_TextBlock_Description_SpecialPerkA,EquipPet_1_TextBlock_Description_SpecialPerkA,EquipPet_2_TextBlock_Description_SpecialPerkA}
        set AllEquipPet_TextBlock_Current_SpecialPerkA = array{EquipPet_0_TextBlock_Current_SpecialPerkA,EquipPet_1_TextBlock_Current_SpecialPerkA,EquipPet_2_TextBlock_Current_SpecialPerkA}
        set AllEquipPet_TextBlock_Current_PlayerPickDamage = array{EquipPet_0_TextBlock_Current_PlayerPickDamage,EquipPet_1_TextBlock_Current_PlayerPickDamage,EquipPet_2_TextBlock_Current_PlayerPickDamage}
        set AllEquipPet_TextBlock_Current_PlayerWeakPoint = array{EquipPet_0_TextBlock_Current_PlayerWeakPoint,EquipPet_1_TextBlock_Current_PlayerWeakPoint,EquipPet_2_TextBlock_Current_PlayerWeakPoint} 
        #Package B
        set AllEquipPet_TextBlock_Description_SpecialPerkB = array{EquipPet_0_TextBlock_Description_SpecialPerkB,EquipPet_1_TextBlock_Description_SpecialPerkB,EquipPet_2_TextBlock_Description_SpecialPerkB}
        set AllEquipPet_TextBlock_Current_SpecialPerkB = array{EquipPet_0_TextBlock_Current_SpecialPerkB,EquipPet_1_TextBlock_Current_SpecialPerkB,EquipPet_2_TextBlock_Current_SpecialPerkB}
        set AllEquipPet_TextBlock_Current_PlayerPickEfficiency = array{EquipPet_0_TextBlock_Current_PlayerPickEfficiency,EquipPet_1_TextBlock_Current_PlayerPickEfficiency,EquipPet_2_TextBlock_Current_PlayerPickEfficiency}
        set AllEquipPet_TextBlock_Current_PlayerPickCriticalDamage = array{EquipPet_0_TextBlock_Current_PlayerPickCriticalDamage,EquipPet_1_TextBlock_Current_PlayerPickCriticalDamage,EquipPet_2_TextBlock_Current_PlayerPickCriticalDamage}

        #Mix
        set AllMixPets_TextureBlock_PetImages = array{MixPets_TextureBlock_PetImages_Box0,MixPets_TextureBlock_PetImages_Box1}
        set AllMixPets_TextBlock_Classes = array{MixPets_TextBlock_Classes_Box0,MixPets_TextBlock_Classes_Box1}
        set AllMixPets_TextBlock_PackA = array{MixPets_TextBlock_PackA_Box0,MixPets_TextBlock_PackA_Box1}
        set AllMixPets_TextBlock_PackB = array{MixPets_TextBlock_PackB_Box0,MixPets_TextBlock_PackB_Box1}
        set AllMixPets_TextBlock_Purity = array{MixPets_TextBlock_Purity_Box0,MixPets_TextBlock_Purity_Box1}
        set AllMixPets_TextBlock_Tier = array{MixPets_TextBlock_Tier_Box0,MixPets_TextBlock_Tier_Box1}
        
        #GeneralSelector
        set AllGeneralSelector_TextureBlock_PetImages = array{GeneralSelector_TextureBlock_PetImages_Box0,GeneralSelector_TextureBlock_PetImages_Box1,texture_block{DefaultImage := Images.EmptyTexture},GeneralSelector_TextureBlock_PetImages_Box2,GeneralSelector_TextureBlock_PetImages_Box3}
        set AllGeneralSelector_TextBlock_Classes = array{GeneralSelector_TextBlock_Classes_Box0,GeneralSelector_TextBlock_Classes_Box1,text_block{},GeneralSelector_TextBlock_Classes_Box2,GeneralSelector_TextBlock_Classes_Box3}
        set AllGeneralSelector_TextBlock_PackA = array{GeneralSelector_TextBlock_PackA_Box0,GeneralSelector_TextBlock_PackA_Box1,text_block{},GeneralSelector_TextBlock_PackA_Box2,GeneralSelector_TextBlock_PackA_Box3}
        set AllGeneralSelector_TextBlock_PackB = array{GeneralSelector_TextBlock_PackB_Box0,GeneralSelector_TextBlock_PackB_Box1,text_block{},GeneralSelector_TextBlock_PackB_Box2,GeneralSelector_TextBlock_PackB_Box3}
        set AllGeneralSelector_TextBlock_Purity = array{GeneralSelector_TextBlock_Purity_Box0,GeneralSelector_TextBlock_Purity_Box1,text_block{},GeneralSelector_TextBlock_Purity_Box2,GeneralSelector_TextBlock_Purity_Box3}
        set AllGeneralSelector_TextBlock_Tier = array{GeneralSelector_TextBlock_Tier_Box0,GeneralSelector_TextBlock_Tier_Box1,text_block{},GeneralSelector_TextBlock_Tier_Box2,GeneralSelector_TextBlock_Tier_Box3}

        #ReciclePet
        set AllRecyclePet_TextureBlock_PetImages = array{RecyclePet_TextureBlock_PetImages_Box0,RecyclePet_TextureBlock_PetImages_Box1,texture_block{DefaultImage := Images.EmptyTexture},RecyclePet_TextureBlock_PetImages_Box2,RecyclePet_TextureBlock_PetImages_Box3}
        set AllRecyclePet_TextBlock_Classes = array{RecyclePet_TextBlock_Classes_Box0,RecyclePet_TextBlock_Classes_Box1,text_block{},RecyclePet_TextBlock_Classes_Box2,RecyclePet_TextBlock_Classes_Box3}
        set AllRecyclePet_TextBlock_PackA = array{RecyclePet_TextBlock_PackA_Box0,RecyclePet_TextBlock_PackA_Box1,text_block{},RecyclePet_TextBlock_PackA_Box2,RecyclePet_TextBlock_PackA_Box3}
        set AllRecyclePet_TextBlock_PackB = array{RecyclePet_TextBlock_PackB_Box0,RecyclePet_TextBlock_PackB_Box1,text_block{},RecyclePet_TextBlock_PackB_Box2,RecyclePet_TextBlock_PackB_Box3}
        set AllRecyclePet_TextBlock_Purity = array{RecyclePet_TextBlock_Purity_Box0,RecyclePet_TextBlock_Purity_Box1,text_block{},RecyclePet_TextBlock_Purity_Box2,RecyclePet_TextBlock_Purity_Box3}
        set AllRecyclePet_TextBlock_Tier = array{RecyclePet_TextBlock_Tier_Box0,RecyclePet_TextBlock_Tier_Box1,text_block{},RecyclePet_TextBlock_Tier_Box2,RecyclePet_TextBlock_Tier_Box3}

        MixPets_Canvas := MakeMixPets_Canvas()
        set MixPets_MaybeCanvas = option{MixPets_Canvas}

        EquipPet_Canvas := MakeEquipPetCanvas()
        set EquipPet_MaybeCanvas = option{EquipPet_Canvas}

        MainMenuPets_Canvas := MakeMainMenuPetsCanvas()
        set MainMenu_MaybeCanvas = option{MainMenuPets_Canvas}

        CreatePet_Canvas := MakeCreatePetCanvas()
        set CreatePet_MaybeCanvas = option{CreatePet_Canvas}

        GeneralSelector_Canvas := MakeGeneralSelectorCanvas()
        set GeneralSelector_MaybeCanvas = option{GeneralSelector_Canvas}

        RecyclePet_Canvas := MakeRecyclePet_Canvas()
        set RecyclePet_MaybeCanvas = option{RecyclePet_Canvas}

        BuySlots_Canvas := MakeBuySlotsCanvas()
        set BuySlots_MaybeCanvas = option{BuySlots_Canvas}

        Subscribes()

    RefreshPetStats(
        InCurrentOption : int,
        AllPets : []pet_data_table,
        MaybePet : ?pet_data_table,
        Selected_TextureBlock_PetImages : texture_block,

        SelectedText_PackageA_LV : text_block,
        SelectedText_PackageB_LV : text_block,
        
        Selected_TextBlock_Current_SpecialPerkA : text_block,
        Selected_TextBlock_Current_SpecialPerkB : text_block,
        Selected_TextBlock_Description_SpecialPerkA : text_block,
        Selected_TextBlock_Description_SpecialPerkB : text_block,
        Selected_TextBlock_Purity : text_block,
        Selected_TextBlock_Class : text_block,
        Selected_TextBlock_PackA : text_block,
        Selected_TextBlock_PackB : text_block,
        
        Selected_TextBlock_Current_TournamentHealth : text_block,
        Selected_TextBlock_Current_TournamentDamage : text_block,
        Selected_TextBlock_Current_TournamentSpeed : text_block,

        Selected_TextBlock_Current_PlayerPickDamage : text_block,
        Selected_TextBlock_Current_PlayerWeakPoint : text_block,

        Selected_TextBlock_Current_TournamentActionSpeed : text_block,
        Selected_TextBlock_Current_TournamentHealingPerk : text_block,
        Selected_TextBlock_Current_TournamentRange : text_block,

        Selected_TextBlock_Current_PlayerPickEfficiency : text_block,
        Selected_TextBlock_Current_PlayerPickCriticalDamage : text_block,

        AllTextureBlock_PetImages : []texture_block,
        AllTextBlock_Classes : []text_block,
        AllTextBlock_PackA : []text_block,
        AllTextBlock_PackB : []text_block,
        AllTextBlock_Purity : []text_block,
        AllTextBlock_Tier : []text_block
        
        ) : void =
            if(Agent := MaybeAgent?,Player := player[Agent], Player.IsActive[]):
                var SelectedEmoji_SpecialA : string = ""
                var SelectedEmoji_SpecialB : string = ""
                if(Pet:=MaybePet?):
                    SelectedText_PackageA_LV.SetText(BasicMessage("PACKAGE A - LEVEL {Pet.LVPet_PackageA}"))
                    SelectedText_PackageB_LV.SetText(BasicMessage("PACKAGE B - LEVEL {Pet.LVPet_PackageB}"))
                    #Package A
                    TupleMaxHealth := CalculePetStat(STAT_TYPE_HEALTH,Pet.LVPet_PackageA,Pet.Type,Pet.Purity)
                    TupleAttackDamage := CalculePetStat(STAT_TYPE_DAMAGE,Pet.LVPet_PackageA,Pet.Type,Pet.Purity)
                    TupleSpeed := CalculePetStat(STAT_TYPE_SPEED,Pet.LVPet_PackageA,Pet.Type,Pet.Purity)

                    TupleSpecialA := CalculePetStat(STAT_TYPE_PLAYER_SPECIAL_A,Pet.LVPet_PackageA,Pet.Type,Pet.Purity)
                    TuplePickDamage := CalculePetStat(STAT_TYPE_PLAYER_PICKDAMAGE,Pet.LVPet_PackageA,Pet.Type,Pet.Purity)
                    TupleWeakPoint := CalculePetStat(STAT_TYPE_PLAYER_WEAKPOINTMULT,Pet.LVPet_PackageA,Pet.Type,Pet.Purity)

                    #Package B
                    TupleAttackSpeed := CalculePetStat(STAT_TYPE_ATTACKSPEED,Pet.LVPet_PackageB,Pet.Type,Pet.Purity)
                    TupleRange := CalculePetStat(STAT_TYPE_RANGE,Pet.LVPet_PackageB,Pet.Type,Pet.Purity)
                    TupleHealingPerk := CalculePetStat(STAT_TYPE_HEALING,Pet.LVPet_PackageB,Pet.Type,Pet.Purity)

                    TupleSpecialB := CalculePetStat(STAT_TYPE_PLAYER_SPECIAL_B,Pet.LVPet_PackageB,Pet.Type,Pet.Purity)
                    TuplePickEfficiency := CalculePetStat(STAT_TYPE_PLAYER_PICKEFFICENCY,Pet.LVPet_PackageB,Pet.Type,Pet.Purity)
                    TuplePickCriticalDamage := CalculePetStat(STAT_TYPE_PLAYER_PICKCRITICALMULT,Pet.LVPet_PackageB,Pet.Type,Pet.Purity)

                    #NextLevel
                    NextMaxHealth := TupleMaxHealth(2)
                    NextAttackDamage := TupleAttackDamage(2)
                    NextSpeed := TupleSpeed(2)

                    #NextLevel
                    NextSpecialA := TupleSpecialA(2)
                    NextPickDamage := TuplePickDamage(2)
                    NextWeakPoint := TupleWeakPoint(2)

                    #NextLevel
                    NextActionSpeed := TupleAttackSpeed(2)
                    NextRange := TupleRange(2)
                    NextHealingPerk := TupleHealingPerk(2)

                    #NextLevel
                    NextSpecialB := TupleSpecialB(2)
                    NextPickEfficiency := TuplePickEfficiency(2)
                    NextCriticalMulti := TuplePickCriticalDamage(2)

                    if(Pet.Type=ADC):
                        set SelectedEmoji_SpecialA = "🎯"
                        set SelectedEmoji_SpecialB = "🏹"
                        Selected_TextBlock_Current_SpecialPerkA.SetText(BasicMessage(GetTripleDecimalMagnitudeLetter(TupleSpecialA(0), " {SelectedEmoji_SpecialA} x", "")))
                        Selected_TextBlock_Current_SpecialPerkB.SetText(BasicMessage(GetDoubleDecimalMagnitudeLetter(TupleSpecialB(0), " {SelectedEmoji_SpecialB}", "s")))
                        Selected_TextBlock_Description_SpecialPerkA.SetText(BasicMessage("Bow damage:"))
                        Selected_TextBlock_Description_SpecialPerkB.SetText(BasicMessage("Arrow craft:"))
                        
                    else if(Pet.Type=DRUID):
                        set SelectedEmoji_SpecialA = "💚"
                        set SelectedEmoji_SpecialB = "💉"
                        Selected_TextBlock_Current_SpecialPerkA.SetText(BasicMessage(GetTripleDecimalMagnitudeLetter(TupleSpecialA(0), "{SelectedEmoji_SpecialA} ", "")))
                        Selected_TextBlock_Current_SpecialPerkB.SetText(BasicMessage(GetTripleDecimalMagnitudeLetter(TupleSpecialB(0), "{SelectedEmoji_SpecialB} ", "")))
                        Selected_TextBlock_Description_SpecialPerkA.SetText(BasicMessage("Pasive heal:"))
                        Selected_TextBlock_Description_SpecialPerkB.SetText(BasicMessage("Emergency heal:"))
                    else if(Pet.Type=TANK):
                        set SelectedEmoji_SpecialA = "💗"
                        set SelectedEmoji_SpecialB = "🛡️"
                        Selected_TextBlock_Current_SpecialPerkA.SetText(BasicMessage(GetTripleDecimalMagnitudeLetter(TupleSpecialA(0), "{SelectedEmoji_SpecialA} ", "")))
                        Selected_TextBlock_Current_SpecialPerkB.SetText(BasicMessage(GetTripleDecimalMagnitudeLetter(TupleSpecialB(0), "{SelectedEmoji_SpecialB} ", "")))
                        Selected_TextBlock_Description_SpecialPerkA.SetText(BasicMessage("Extra health:"))
                        Selected_TextBlock_Description_SpecialPerkB.SetText(BasicMessage("Combat shield:"))
                    else if(Pet.Type=FIGHTER):
                        set SelectedEmoji_SpecialA = "🔪"
                        set SelectedEmoji_SpecialB = "🗡️"
                        Selected_TextBlock_Current_SpecialPerkA.SetText(BasicMessage(GetTripleDecimalMagnitudeLetter(TupleSpecialA(0), "{SelectedEmoji_SpecialA} x", "")))
                        Selected_TextBlock_Current_SpecialPerkB.SetText(BasicMessage(GetTripleDecimalMagnitudeLetter(TupleSpecialB(0), "{SelectedEmoji_SpecialB} x", "")))
                        Selected_TextBlock_Description_SpecialPerkA.SetText(BasicMessage("Melee damage:"))
                        Selected_TextBlock_Description_SpecialPerkB.SetText(BasicMessage("Melee UltraStrike:"))
                    else if(Pet.Type=HUNTER):
                        set SelectedEmoji_SpecialA = "🐾"
                        set SelectedEmoji_SpecialB = "💢"
                        Selected_TextBlock_Current_SpecialPerkA.SetText(BasicMessage(GetTripleDecimalMagnitudeLetter(TupleSpecialA(0), "{SelectedEmoji_SpecialA} x", "")))
                        Selected_TextBlock_Current_SpecialPerkB.SetText(BasicMessage(GetTripleDecimalMagnitudeLetter(TupleSpecialB(0), "{SelectedEmoji_SpecialB} x", "")))
                        Selected_TextBlock_Description_SpecialPerkA.SetText(BasicMessage("Hunt probability:"))
                        Selected_TextBlock_Description_SpecialPerkB.SetText(BasicMessage("Weapon critical:"))
                        

                    TuplePurity:=ConvertRawPurity(Pet.Purity)
                    SetPetImage(Pet.Type, Selected_TextureBlock_PetImages)
                    SetPetPurityAndTier(TuplePurity(1),TuplePurity(0),Selected_TextBlock_Purity)
                    SetPetClass(Pet,Selected_TextBlock_Class, Selected_TextBlock_PackA, Selected_TextBlock_PackB)

                    #ALL STATS
                    #Package A
                    Selected_TextBlock_Current_TournamentHealth.SetText(BasicMessage(GetDoubleDecimalMagnitudeLetter(TupleMaxHealth(0), "❤️ ", "")))
                    Selected_TextBlock_Current_TournamentDamage.SetText(BasicMessage(GetDoubleDecimalMagnitudeLetter(TupleAttackDamage(0), "💥 ", "")))
                    Selected_TextBlock_Current_TournamentSpeed.SetText(BasicMessage(GetDoubleDecimalMagnitudeLetter(TupleSpeed(0), "🏃 ", "s / tile")))

                    Selected_TextBlock_Current_PlayerPickDamage.SetText(BasicMessage(GetTripleDecimalMagnitudeLetter(TuplePickDamage(0), "🔨 x", "")))
                    Selected_TextBlock_Current_PlayerWeakPoint.SetText(BasicMessage(GetTripleDecimalMagnitudeLetter(TupleWeakPoint(0), "🧨 x", "")))            
                    
                    Selected_TextBlock_Current_TournamentActionSpeed.SetText(BasicMessage(GetDoubleDecimalMagnitudeLetter(TupleAttackSpeed(0), "🔥", "s")))
                    Selected_TextBlock_Current_TournamentRange.SetText(BasicMessage(GetMagnitudeLetter(TupleRange(0), "🏹 ", " tiles")))

                    if(Pet.Type = DRUID):
                        Selected_TextBlock_Current_TournamentHealingPerk.SetText(BasicMessage(GetDoubleDecimalMagnitudeLetter(TupleHealingPerk(0), "💉 ", "")))
                    else:
                        Selected_TextBlock_Current_TournamentHealingPerk.SetText(BasicMessage("-"))

                    Selected_TextBlock_Current_PlayerPickEfficiency.SetText(BasicMessage(GetTripleDecimalMagnitudeLetter(TuplePickEfficiency(0), "🚀 x", "")))
                    Selected_TextBlock_Current_PlayerPickCriticalDamage.SetText(BasicMessage(GetTripleDecimalMagnitudeLetter(TuplePickCriticalDamage(0), "💥 x", "")))
                else:
                    SelectedText_PackageA_LV.SetText(BasicMessage("-"))
                    SelectedText_PackageB_LV.SetText(BasicMessage("-"))

                    SelectedText_PackageA_LV.SetTextColor(White)
                    SelectedText_PackageB_LV.SetTextColor(White)

                    Selected_TextBlock_Current_SpecialPerkA.SetText(BasicMessage("-"))
                    Selected_TextBlock_Current_SpecialPerkB.SetText(BasicMessage("-"))
                    SetPetImage(-1, Selected_TextureBlock_PetImages)
                    
                    Selected_TextBlock_PackA.SetText(BasicMessage("-"))
                    Selected_TextBlock_PackB.SetText(BasicMessage("-"))

                    Selected_TextBlock_PackA.SetTextColor(White)
                    Selected_TextBlock_PackB.SetTextColor(White)

                    Selected_TextBlock_Purity.SetText(BasicMessage("-"))
                    Selected_TextBlock_Class.SetText(BasicMessage("-"))

                    Selected_TextBlock_Current_TournamentHealth.SetText(BasicMessage("❤️ -"))
                    Selected_TextBlock_Current_TournamentDamage.SetText(BasicMessage("💥 -"))
                    Selected_TextBlock_Current_TournamentSpeed.SetText(BasicMessage("🏃 -"))

                    Selected_TextBlock_Current_PlayerPickDamage.SetText(BasicMessage("🔨 -"))
                    Selected_TextBlock_Current_PlayerWeakPoint.SetText(BasicMessage("🧨 -"))            
                    
                    Selected_TextBlock_Current_TournamentActionSpeed.SetText(BasicMessage("🔥 -"))
                    Selected_TextBlock_Current_TournamentRange.SetText(BasicMessage("🏹 -"))

                    Selected_TextBlock_Current_TournamentHealingPerk.SetText(BasicMessage("-"))

                    Selected_TextBlock_Current_PlayerPickEfficiency.SetText(BasicMessage("🚀 -"))
                    Selected_TextBlock_Current_PlayerPickCriticalDamage.SetText(BasicMessage("💥 -"))

                #The other images
                for(Index:=0..AllTextureBlock_PetImages.Length-1):
                    if:
                        TextureBlock_PetImage:=AllTextureBlock_PetImages[Index]
                        TextBlock_Class:=AllTextBlock_Classes[Index]
                        TextBlock_PackA:=AllTextBlock_PackA[Index]
                        TextBlock_PackB:=AllTextBlock_PackB[Index]
                        TextBlock_Purity:=AllTextBlock_Purity[Index]
                        TextBlock_Tier:=AllTextBlock_Tier[Index]
                    then:
                        if(PetIndexSelected:=AllPets[InCurrentOption-2+Index]):
                            InTuplePurity:=ConvertRawPurity(PetIndexSelected.Purity)
                            SetPetImage(PetIndexSelected.Type, TextureBlock_PetImage)
                            SetPetClass(PetIndexSelected, TextBlock_Class,TextBlock_PackA,TextBlock_PackB)
                            SetPetPurity(InTuplePurity(1), TextBlock_Purity)
                            SetPetTier(InTuplePurity(0), TextBlock_Tier)
                        else:
                            TextureBlock_PetImage.SetImage(Images.Cross_Icon)
                            TextBlock_Class.SetText(BasicMessage("-"))
                            TextBlock_Purity.SetText(BasicMessage("-"))
                            TextBlock_Tier.SetText(BasicMessage("-"))
                            TextBlock_PackA.SetText(BasicMessage("-"))
                            TextBlock_PackA.SetTextColor(White)
                            TextBlock_PackB.SetText(BasicMessage("-"))
                            TextBlock_PackB.SetTextColor(White)
                            TextureBlock_PetImage.SetTint(White)
                            TextBlock_Class.SetTextColor(White)
                            TextBlock_Purity.SetTextColor(White)
                            TextBlock_Tier.SetTextColor(White)
                        #Pet not found!
    ################
    #EQUIP MENU CANVAS
    MakeEquipPetCanvas(): canvas=
        Canvas := canvas:
            Slots:=array:
                #Main overlay
                canvas_slot:
                    Anchors:=anchors{Minimum:=vector2{X:=0.5,Y:=0.5},Maximum:=vector2{X:=0.5,Y:=0.5}}
                    ZOrder:=0
                    SizeToContent:=true
                    Offsets:=margin{Left:=0.0,Top:=0.0,Right:=0.0,Bottom:=0.0} #PositionX, PositionY, SizeX, SizeY
                    Alignment:=vector2{X:=0.5,Y:=0.5}
                    Widget:=texture_block{DefaultDesiredSize:=vector2{X:=3000.0,Y:=1250.0},DefaultImage:=Images.Img_MainBackground} 
                canvas_slot:
                    Anchors:=anchors{Minimum:=vector2{X:=0.5,Y:=0.5},Maximum:=vector2{X:=0.5,Y:=0.5}}
                    ZOrder:=0
                    SizeToContent:=true
                    Offsets:=margin{Left:=0.0,Top:=0.0,Right:=0.0,Bottom:=0.0} #PositionX, PositionY, SizeX, SizeY
                    Alignment:=vector2{X:=0.5,Y:=0.5}
                    Widget:=stack_box:  
                        Orientation:=orientation.Horizontal
                            Slots:=array:                                
                                stack_box_slot: 
                                    HorizontalAlignment:=horizontal_alignment.Center
                                    VerticalAlignment:=vertical_alignment.Center
                                    Padding:=RowPadding
                                    Widget:=stack_box:  #COLUMN 0
                                        Orientation:=orientation.Vertical
                                            Slots:=array:
                                                stack_box_slot:
                                                    Padding:=RowPadding
                                                    VerticalAlignment:=vertical_alignment.Center
                                                    HorizontalAlignment:=horizontal_alignment.Center
                                                    Widget:=EquipPet_0_SlotInfo
                                                stack_box_slot: 
                                                    Padding:=margin{Right:=SuperLowPadding,Left:=SuperLowPadding}
                                                    HorizontalAlignment:=horizontal_alignment.Center
                                                    VerticalAlignment:=vertical_alignment.Center
                                                    Widget:=stack_box:  
                                                        Orientation:=orientation.Vertical
                                                        Slots:=array:
                                                            stack_box_slot: 
                                                                Padding:=margin{Top:=SelectorBoxPadding,Bottom:=SelectorBoxPadding}
                                                                Widget:=overlay:
                                                                    Slots:=array:
                                                                        overlay_slot:
                                                                            HorizontalAlignment:=horizontal_alignment.Fill
                                                                            VerticalAlignment:=vertical_alignment.Fill
                                                                            Widget:=texture_block{DefaultImage:=Images.Img_LineBackground}
                                                                        overlay_slot:
                                                                            HorizontalAlignment:=horizontal_alignment.Fill
                                                                            VerticalAlignment:=vertical_alignment.Fill
                                                                            Padding:=margin{Left:=SelectorBoxPadding,Right:=SelectorBoxPadding,Top:=SelectorBoxPadding,Bottom:=SelectorBoxPadding}
                                                                            Widget:=texture_block{DefaultImage:=Images.Img_LineBackground}
                                                                        overlay_slot:
                                                                            HorizontalAlignment:=horizontal_alignment.Center
                                                                            VerticalAlignment:=vertical_alignment.Center
                                                                            Widget:=texture_block{DefaultImage:=Images.EmptyTexture,DefaultDesiredSize:=vector2{X:=600.0,Y:=615.0}}
                                                                        overlay_slot:
                                                                            HorizontalAlignment:=horizontal_alignment.Center
                                                                            VerticalAlignment:=vertical_alignment.Center
                                                                            Widget:=stack_box:  
                                                                                Orientation:=orientation.Vertical
                                                                                Slots:=array:
                                                                                    stack_box_slot: 
                                                                                        Widget:=stack_box:  
                                                                                            Orientation:=orientation.Horizontal
                                                                                            Slots:=array:
                                                                                                stack_box_slot: 
                                                                                                    Padding:=RowPadding
                                                                                                    Widget:=EquipPet_0_TextureBlock_PetImages
                                                                                                stack_box_slot: 
                                                                                                    HorizontalAlignment:=horizontal_alignment.Center
                                                                                                    VerticalAlignment:=vertical_alignment.Center
                                                                                                    Widget:=stack_box:  
                                                                                                        Orientation:=orientation.Vertical
                                                                                                        Slots:=array:
                                                                                                            stack_box_slot: 
                                                                                                                HorizontalAlignment:=horizontal_alignment.Center
                                                                                                                VerticalAlignment:=vertical_alignment.Center
                                                                                                                #Padding:=margin{Left:=TextPetsBoxPadding}
                                                                                                                Widget:=stack_box:
                                                                                                                    Orientation:=orientation.Horizontal
                                                                                                                    Slots:=array:
                                                                                                                        stack_box_slot: 
                                                                                                                            Padding:=margin{Left:=2.5,Right:=2.5}
                                                                                                                            Widget:=EquipPet_0_TextBlock_Class
                                                                                                                        stack_box_slot: 
                                                                                                                            Padding:=margin{Left:=2.5, Right:=1.0}
                                                                                                                            Widget:=EquipPet_0_TextBlock_PackA
                                                                                                                        stack_box_slot: 
                                                                                                                            Padding:=margin{Right:=2.5}
                                                                                                                            Widget:=EquipPet_0_TextBlock_PackB
                                                                                                            stack_box_slot: 
                                                                                                                Padding:=RowPadding
                                                                                                                Widget:=EquipPet_0_TextBlock_Purity
                                                                                    stack_box_slot: 
                                                                                        HorizontalAlignment:=horizontal_alignment.Fill
                                                                                        VerticalAlignment:=vertical_alignment.Fill
                                                                                        Widget:=texture_block{DefaultImage:=Images.Img_HorizontalLine,DefaultDesiredSize:=vector2{Y:=3.0}}
                                                                                    stack_box_slot:  
                                                                                        HorizontalAlignment:=horizontal_alignment.Center
                                                                                        VerticalAlignment:=vertical_alignment.Center
                                                                                        Padding:=RowPadding
                                                                                        Widget:=EquipPet_0_Text_PackageA_LV
                                                                                    stack_box_slot: 
                                                                                        HorizontalAlignment:=horizontal_alignment.Fill
                                                                                        VerticalAlignment:=vertical_alignment.Fill
                                                                                        Widget:=texture_block{DefaultImage:=Images.Img_HorizontalLine,DefaultDesiredSize:=vector2{Y:=3.0}}
                                                                                    stack_box_slot: 
                                                                                        HorizontalAlignment:=horizontal_alignment.Fill
                                                                                        VerticalAlignment:=vertical_alignment.Fill
                                                                                        Widget:=stack_box:  
                                                                                            Orientation:=orientation.Horizontal
                                                                                            Slots:=array:
                                                                                                stack_box_slot: 
                                                                                                    Widget:=stack_box:
                                                                                                        Orientation:=orientation.Vertical
                                                                                                        Slots:=array:
                                                                                                stack_box_slot: 
                                                                                                    Widget:=stack_box:
                                                                                                        Orientation:=orientation.Vertical
                                                                                                        Slots:=array:
                                                                                                            stack_box_slot: 
                                                                                                                Padding:=RowPadding
                                                                                                                Widget:=text_block{DefaultText:=BasicMessage("Player enhancements"), DefaultTextColor:=Cyan, DefaultShadowOffset:=ShadowOffset, DefaultShadowOpacity := DefaultTextShadowOpacity}
                                                                                                            stack_box_slot: 
                                                                                                                HorizontalAlignment:=horizontal_alignment.Fill
                                                                                                                VerticalAlignment:=vertical_alignment.Fill
                                                                                                                Widget:=stack_box:  
                                                                                                                    Orientation:=orientation.Horizontal
                                                                                                                    Slots:=array:
                                                                                                                        stack_box_slot: 
                                                                                                                            HorizontalAlignment:=horizontal_alignment.Fill
                                                                                                                            VerticalAlignment:=vertical_alignment.Fill
                                                                                                                            Widget:=stack_box:  
                                                                                                                                Orientation:=orientation.Vertical
                                                                                                                                Slots:=array:
                                                                                                                                    stack_box_slot: 
                                                                                                                                        Padding:=RowPadding
                                                                                                                                        HorizontalAlignment:=horizontal_alignment.Left
                                                                                                                                        Widget:=EquipPet_0_TextBlock_Description_SpecialPerkA
                                                                                                                                    stack_box_slot: 
                                                                                                                                        Padding:=RowPadding
                                                                                                                                        HorizontalAlignment:=horizontal_alignment.Left
                                                                                                                                        Widget:=text_block{DefaultText:=BasicMessage("Pick damage:"), DefaultTextColor:=White, DefaultShadowOffset:=ShadowOffset, DefaultShadowOpacity := DefaultTextShadowOpacity}
                                                                                                                                    stack_box_slot: 
                                                                                                                                        Padding:=RowPadding
                                                                                                                                        HorizontalAlignment:=horizontal_alignment.Left
                                                                                                                                        Widget:=text_block{DefaultText:=BasicMessage("Weak point mult:"), DefaultTextColor:=White, DefaultShadowOffset:=ShadowOffset, DefaultShadowOpacity := DefaultTextShadowOpacity}
                                                                                                                        stack_box_slot: 
                                                                                                                            HorizontalAlignment:=horizontal_alignment.Fill
                                                                                                                            VerticalAlignment:=vertical_alignment.Fill
                                                                                                                            Widget:=stack_box:  
                                                                                                                                Orientation:=orientation.Vertical
                                                                                                                                Slots:=array:
                                                                                                                                    stack_box_slot: 
                                                                                                                                        Padding:=RowPadding
                                                                                                                                        Widget:=EquipPet_0_TextBlock_Current_SpecialPerkA
                                                                                                                                    stack_box_slot: 
                                                                                                                                        Padding:=RowPadding
                                                                                                                                        Widget:=EquipPet_0_TextBlock_Current_PlayerPickDamage
                                                                                                                                    stack_box_slot:
                                                                                                                                        Padding:=RowPadding 
                                                                                                                                        Widget:=EquipPet_0_TextBlock_Current_PlayerWeakPoint
                                                                                    stack_box_slot: 
                                                                                        HorizontalAlignment:=horizontal_alignment.Fill
                                                                                        VerticalAlignment:=vertical_alignment.Fill
                                                                                        Widget:=texture_block{DefaultImage:=Images.Img_HorizontalLine,DefaultDesiredSize:=vector2{Y:=3.0}}
                                                                                    stack_box_slot:  
                                                                                        HorizontalAlignment:=horizontal_alignment.Center
                                                                                        VerticalAlignment:=vertical_alignment.Center
                                                                                        Padding:=RowPadding
                                                                                        Widget:=EquipPet_0_Text_PackageB_LV
                                                                                    stack_box_slot: 
                                                                                        HorizontalAlignment:=horizontal_alignment.Fill
                                                                                        VerticalAlignment:=vertical_alignment.Fill
                                                                                        Widget:=texture_block{DefaultImage:=Images.Img_HorizontalLine,DefaultDesiredSize:=vector2{Y:=3.0}}
                                                                                    stack_box_slot: 
                                                                                        HorizontalAlignment:=horizontal_alignment.Fill
                                                                                        VerticalAlignment:=vertical_alignment.Fill
                                                                                        Widget:=stack_box:  
                                                                                            Orientation:=orientation.Horizontal
                                                                                            Slots:=array:
                                                                                                stack_box_slot: 
                                                                                                    Widget:=stack_box:
                                                                                                        Orientation:=orientation.Vertical
                                                                                                        Slots:=array:
                                                                                                            stack_box_slot: 
                                                                                                                Padding:=RowPadding
                                                                                                                Widget:=text_block{DefaultText:=BasicMessage("Player enhancements"), DefaultTextColor:=Cyan, DefaultShadowOffset:=ShadowOffset, DefaultShadowOpacity := DefaultTextShadowOpacity}
                                                                                                            stack_box_slot: 
                                                                                                                HorizontalAlignment:=horizontal_alignment.Fill
                                                                                                                VerticalAlignment:=vertical_alignment.Fill
                                                                                                                Widget:=stack_box:  
                                                                                                                    Orientation:=orientation.Horizontal
                                                                                                                    Slots:=array:
                                                                                                                        stack_box_slot: 
                                                                                                                            HorizontalAlignment:=horizontal_alignment.Fill
                                                                                                                            VerticalAlignment:=vertical_alignment.Fill
                                                                                                                            Widget:=stack_box:  
                                                                                                                                Orientation:=orientation.Vertical
                                                                                                                                Slots:=array:
                                                                                                                                    stack_box_slot: 
                                                                                                                                        Padding:=RowPadding
                                                                                                                                        HorizontalAlignment:=horizontal_alignment.Left
                                                                                                                                        Widget:=EquipPet_0_TextBlock_Description_SpecialPerkB
                                                                                                                                    stack_box_slot: 
                                                                                                                                        Padding:=RowPadding
                                                                                                                                        HorizontalAlignment:=horizontal_alignment.Left
                                                                                                                                        Widget:=text_block{DefaultText:=BasicMessage("Pick efficiency:"), DefaultTextColor:=White, DefaultShadowOffset:=ShadowOffset, DefaultShadowOpacity := DefaultTextShadowOpacity}
                                                                                                                                    stack_box_slot: 
                                                                                                                                        Padding:=RowPadding
                                                                                                                                        HorizontalAlignment:=horizontal_alignment.Left
                                                                                                                                        Widget:=text_block{DefaultText:=BasicMessage("Pick critical dam:"), DefaultTextColor:=White, DefaultShadowOffset:=ShadowOffset, DefaultShadowOpacity := DefaultTextShadowOpacity}
                                                                                                                        stack_box_slot: 
                                                                                                                            HorizontalAlignment:=horizontal_alignment.Fill
                                                                                                                            VerticalAlignment:=vertical_alignment.Fill
                                                                                                                            Widget:=stack_box:  
                                                                                                                                Orientation:=orientation.Vertical
                                                                                                                                Slots:=array:
                                                                                                                                    stack_box_slot: 
                                                                                                                                        Padding:=RowPadding
                                                                                                                                        Widget:=EquipPet_0_TextBlock_Current_SpecialPerkB
                                                                                                                                    stack_box_slot: 
                                                                                                                                        Padding:=RowPadding
                                                                                                                                        Widget:=EquipPet_0_TextBlock_Current_PlayerPickEfficiency
                                                                                                                                    stack_box_slot: 
                                                                                                                                        Padding:=RowPadding
                                                                                                                                        Widget:=EquipPet_0_TextBlock_Current_PlayerPickCriticalDamage
                                                stack_box_slot:
                                                    Padding:=RowPadding
                                                    VerticalAlignment:=vertical_alignment.Center
                                                    HorizontalAlignment:=horizontal_alignment.Center
                                                    Widget:=overlay:
                                                        Slots:=array:
                                                            overlay_slot:
                                                                HorizontalAlignment:=horizontal_alignment.Center
                                                                VerticalAlignment:=vertical_alignment.Center
                                                                Widget:=texture_block{DefaultImage:=Images.EmptyTexture,DefaultDesiredSize:=vector2{X:=500.0,Y:=185.0}}
                                                            overlay_slot:
                                                                HorizontalAlignment:=horizontal_alignment.Fill
                                                                VerticalAlignment:=vertical_alignment.Fill
                                                                Widget:=EquipPet_0_ChangeSlotButton

                                stack_box_slot: 
                                    HorizontalAlignment:=horizontal_alignment.Center
                                    VerticalAlignment:=vertical_alignment.Center
                                    Padding:=RowPadding
                                    Widget:=stack_box:  #COLUMN 1
                                        Orientation:=orientation.Vertical
                                            Slots:=array:
                                                stack_box_slot: 
                                                    Padding:=RowPadding
                                                    VerticalAlignment:=vertical_alignment.Center
                                                    HorizontalAlignment:=horizontal_alignment.Center
                                                    Widget:=EquipPet_1_SlotInfo
                                                stack_box_slot: 
                                                    Padding:=margin{Right:=SuperLowPadding,Left:=SuperLowPadding}
                                                    HorizontalAlignment:=horizontal_alignment.Center
                                                    VerticalAlignment:=vertical_alignment.Center
                                                    Widget:=stack_box:  
                                                        Orientation:=orientation.Vertical
                                                        Slots:=array:
                                                            stack_box_slot: 
                                                                Padding:=margin{Top:=SelectorBoxPadding,Bottom:=SelectorBoxPadding}
                                                                Widget:=overlay:
                                                                    Slots:=array:
                                                                        overlay_slot:
                                                                            HorizontalAlignment:=horizontal_alignment.Fill
                                                                            VerticalAlignment:=vertical_alignment.Fill
                                                                            Widget:=texture_block{DefaultImage:=Images.Img_LineBackground}
                                                                        overlay_slot:
                                                                            HorizontalAlignment:=horizontal_alignment.Fill
                                                                            VerticalAlignment:=vertical_alignment.Fill
                                                                            Padding:=margin{Left:=SelectorBoxPadding,Right:=SelectorBoxPadding,Top:=SelectorBoxPadding,Bottom:=SelectorBoxPadding}
                                                                            Widget:=texture_block{DefaultImage:=Images.Img_LineBackground}
                                                                        overlay_slot:
                                                                            HorizontalAlignment:=horizontal_alignment.Center
                                                                            VerticalAlignment:=vertical_alignment.Center
                                                                            Widget:=texture_block{DefaultImage:=Images.EmptyTexture,DefaultDesiredSize:=vector2{X:=600.0,Y:=615.0}}
                                                                        overlay_slot:
                                                                            HorizontalAlignment:=horizontal_alignment.Center
                                                                            VerticalAlignment:=vertical_alignment.Center
                                                                            Widget:=stack_box:  
                                                                                Orientation:=orientation.Vertical
                                                                                Slots:=array:
                                                                                    stack_box_slot: 
                                                                                        Widget:=stack_box:  
                                                                                            Orientation:=orientation.Horizontal
                                                                                            Slots:=array:
                                                                                                stack_box_slot: 
                                                                                                    Padding:=RowPadding
                                                                                                    Widget:=EquipPet_1_TextureBlock_PetImages
                                                                                                stack_box_slot: 
                                                                                                    HorizontalAlignment:=horizontal_alignment.Center
                                                                                                    VerticalAlignment:=vertical_alignment.Center
                                                                                                    Widget:=stack_box:  
                                                                                                        Orientation:=orientation.Vertical
                                                                                                        Slots:=array:
                                                                                                            stack_box_slot: 
                                                                                                                HorizontalAlignment:=horizontal_alignment.Center
                                                                                                                VerticalAlignment:=vertical_alignment.Center
                                                                                                                #Padding:=margin{Left:=TextPetsBoxPadding}
                                                                                                                Widget:=stack_box:
                                                                                                                    Orientation:=orientation.Horizontal
                                                                                                                    Slots:=array:
                                                                                                                        stack_box_slot: 
                                                                                                                            Padding:=margin{Left:=2.5,Right:=2.5}
                                                                                                                            Widget:=EquipPet_1_TextBlock_Class
                                                                                                                        stack_box_slot: 
                                                                                                                            Padding:=margin{Left:=2.5, Right:=1.0}
                                                                                                                            Widget:=EquipPet_1_TextBlock_PackA
                                                                                                                        stack_box_slot: 
                                                                                                                            Padding:=margin{Right:=2.5}
                                                                                                                            Widget:=EquipPet_1_TextBlock_PackB
                                                                                                            stack_box_slot: 
                                                                                                                Padding:=RowPadding
                                                                                                                Widget:=EquipPet_1_TextBlock_Purity
                                                                                    stack_box_slot: 
                                                                                        HorizontalAlignment:=horizontal_alignment.Fill
                                                                                        VerticalAlignment:=vertical_alignment.Fill
                                                                                        Widget:=texture_block{DefaultImage:=Images.Img_HorizontalLine,DefaultDesiredSize:=vector2{Y:=3.0}}
                                                                                    stack_box_slot:  
                                                                                        HorizontalAlignment:=horizontal_alignment.Center
                                                                                        VerticalAlignment:=vertical_alignment.Center
                                                                                        Padding:=RowPadding
                                                                                        Widget:=EquipPet_1_Text_PackageA_LV
                                                                                    stack_box_slot: 
                                                                                        HorizontalAlignment:=horizontal_alignment.Fill
                                                                                        VerticalAlignment:=vertical_alignment.Fill
                                                                                        Widget:=texture_block{DefaultImage:=Images.Img_HorizontalLine,DefaultDesiredSize:=vector2{Y:=3.0}}
                                                                                    stack_box_slot: 
                                                                                        HorizontalAlignment:=horizontal_alignment.Fill
                                                                                        VerticalAlignment:=vertical_alignment.Fill
                                                                                        Widget:=stack_box:  
                                                                                            Orientation:=orientation.Horizontal
                                                                                            Slots:=array:
                                                                                                stack_box_slot: 
                                                                                                    Widget:=stack_box:
                                                                                                        Orientation:=orientation.Vertical
                                                                                                        Slots:=array:
                                                                                                stack_box_slot: 
                                                                                                    Widget:=stack_box:
                                                                                                        Orientation:=orientation.Vertical
                                                                                                        Slots:=array:
                                                                                                            stack_box_slot: 
                                                                                                                Padding:=RowPadding
                                                                                                                Widget:=text_block{DefaultText:=BasicMessage("Player enhancements"), DefaultTextColor:=Cyan, DefaultShadowOffset:=ShadowOffset, DefaultShadowOpacity := DefaultTextShadowOpacity}
                                                                                                            stack_box_slot: 
                                                                                                                HorizontalAlignment:=horizontal_alignment.Fill
                                                                                                                VerticalAlignment:=vertical_alignment.Fill
                                                                                                                Widget:=stack_box:  
                                                                                                                    Orientation:=orientation.Horizontal
                                                                                                                    Slots:=array:
                                                                                                                        stack_box_slot: 
                                                                                                                            HorizontalAlignment:=horizontal_alignment.Fill
                                                                                                                            VerticalAlignment:=vertical_alignment.Fill
                                                                                                                            Widget:=stack_box:  
                                                                                                                                Orientation:=orientation.Vertical
                                                                                                                                Slots:=array:
                                                                                                                                    stack_box_slot: 
                                                                                                                                        Padding:=RowPadding
                                                                                                                                        HorizontalAlignment:=horizontal_alignment.Left
                                                                                                                                        Widget:=EquipPet_1_TextBlock_Description_SpecialPerkA
                                                                                                                                    stack_box_slot: 
                                                                                                                                        Padding:=RowPadding
                                                                                                                                        HorizontalAlignment:=horizontal_alignment.Left
                                                                                                                                        Widget:=text_block{DefaultText:=BasicMessage("Pick damage:"), DefaultTextColor:=White, DefaultShadowOffset:=ShadowOffset, DefaultShadowOpacity := DefaultTextShadowOpacity}
                                                                                                                                    stack_box_slot: 
                                                                                                                                        Padding:=RowPadding
                                                                                                                                        HorizontalAlignment:=horizontal_alignment.Left
                                                                                                                                        Widget:=text_block{DefaultText:=BasicMessage("Weak point mult:"), DefaultTextColor:=White, DefaultShadowOffset:=ShadowOffset, DefaultShadowOpacity := DefaultTextShadowOpacity}
                                                                                                                        stack_box_slot: 
                                                                                                                            HorizontalAlignment:=horizontal_alignment.Fill
                                                                                                                            VerticalAlignment:=vertical_alignment.Fill
                                                                                                                            Widget:=stack_box:  
                                                                                                                                Orientation:=orientation.Vertical
                                                                                                                                Slots:=array:
                                                                                                                                    stack_box_slot: 
                                                                                                                                        Padding:=RowPadding
                                                                                                                                        Widget:=EquipPet_1_TextBlock_Current_SpecialPerkA
                                                                                                                                    stack_box_slot: 
                                                                                                                                        Padding:=RowPadding
                                                                                                                                        Widget:=EquipPet_1_TextBlock_Current_PlayerPickDamage
                                                                                                                                    stack_box_slot:
                                                                                                                                        Padding:=RowPadding 
                                                                                                                                        Widget:=EquipPet_1_TextBlock_Current_PlayerWeakPoint
                                                                                    stack_box_slot: 
                                                                                        HorizontalAlignment:=horizontal_alignment.Fill
                                                                                        VerticalAlignment:=vertical_alignment.Fill
                                                                                        Widget:=texture_block{DefaultImage:=Images.Img_HorizontalLine,DefaultDesiredSize:=vector2{Y:=3.0}}
                                                                                    stack_box_slot:  
                                                                                        HorizontalAlignment:=horizontal_alignment.Center
                                                                                        VerticalAlignment:=vertical_alignment.Center
                                                                                        Padding:=RowPadding
                                                                                        Widget:=EquipPet_1_Text_PackageB_LV
                                                                                    stack_box_slot: 
                                                                                        HorizontalAlignment:=horizontal_alignment.Fill
                                                                                        VerticalAlignment:=vertical_alignment.Fill
                                                                                        Widget:=texture_block{DefaultImage:=Images.Img_HorizontalLine,DefaultDesiredSize:=vector2{Y:=3.0}}
                                                                                    stack_box_slot: 
                                                                                        HorizontalAlignment:=horizontal_alignment.Fill
                                                                                        VerticalAlignment:=vertical_alignment.Fill
                                                                                        Widget:=stack_box:  
                                                                                            Orientation:=orientation.Horizontal
                                                                                            Slots:=array:
                                                                                                stack_box_slot: 
                                                                                                    Widget:=stack_box:
                                                                                                        Orientation:=orientation.Vertical
                                                                                                        Slots:=array:
                                                                                                            stack_box_slot: 
                                                                                                                Padding:=RowPadding
                                                                                                                Widget:=text_block{DefaultText:=BasicMessage("Player enhancements"), DefaultTextColor:=Cyan, DefaultShadowOffset:=ShadowOffset, DefaultShadowOpacity := DefaultTextShadowOpacity}
                                                                                                            stack_box_slot: 
                                                                                                                HorizontalAlignment:=horizontal_alignment.Fill
                                                                                                                VerticalAlignment:=vertical_alignment.Fill
                                                                                                                Widget:=stack_box:  
                                                                                                                    Orientation:=orientation.Horizontal
                                                                                                                    Slots:=array:
                                                                                                                        stack_box_slot: 
                                                                                                                            HorizontalAlignment:=horizontal_alignment.Fill
                                                                                                                            VerticalAlignment:=vertical_alignment.Fill
                                                                                                                            Widget:=stack_box:  
                                                                                                                                Orientation:=orientation.Vertical
                                                                                                                                Slots:=array:
                                                                                                                                    stack_box_slot: 
                                                                                                                                        Padding:=RowPadding
                                                                                                                                        HorizontalAlignment:=horizontal_alignment.Left
                                                                                                                                        Widget:=EquipPet_1_TextBlock_Description_SpecialPerkB
                                                                                                                                    stack_box_slot: 
                                                                                                                                        Padding:=RowPadding
                                                                                                                                        HorizontalAlignment:=horizontal_alignment.Left
                                                                                                                                        Widget:=text_block{DefaultText:=BasicMessage("Pick efficiency:"), DefaultTextColor:=White, DefaultShadowOffset:=ShadowOffset, DefaultShadowOpacity := DefaultTextShadowOpacity}
                                                                                                                                    stack_box_slot: 
                                                                                                                                        Padding:=RowPadding
                                                                                                                                        HorizontalAlignment:=horizontal_alignment.Left
                                                                                                                                        Widget:=text_block{DefaultText:=BasicMessage("Pick critical dam:"), DefaultTextColor:=White, DefaultShadowOffset:=ShadowOffset, DefaultShadowOpacity := DefaultTextShadowOpacity}
                                                                                                                        stack_box_slot: 
                                                                                                                            HorizontalAlignment:=horizontal_alignment.Fill
                                                                                                                            VerticalAlignment:=vertical_alignment.Fill
                                                                                                                            Widget:=stack_box:  
                                                                                                                                Orientation:=orientation.Vertical
                                                                                                                                Slots:=array:
                                                                                                                                    stack_box_slot: 
                                                                                                                                        Padding:=RowPadding
                                                                                                                                        Widget:=EquipPet_1_TextBlock_Current_SpecialPerkB
                                                                                                                                    stack_box_slot: 
                                                                                                                                        Padding:=RowPadding
                                                                                                                                        Widget:=EquipPet_1_TextBlock_Current_PlayerPickEfficiency
                                                                                                                                    stack_box_slot: 
                                                                                                                                        Padding:=RowPadding
                                                                                                                                        Widget:=EquipPet_1_TextBlock_Current_PlayerPickCriticalDamage                                            
                                                stack_box_slot:
                                                    Padding:=RowPadding
                                                    VerticalAlignment:=vertical_alignment.Center
                                                    HorizontalAlignment:=horizontal_alignment.Center
                                                    Widget:=overlay:
                                                        Slots:=array:
                                                            overlay_slot:
                                                                HorizontalAlignment:=horizontal_alignment.Center
                                                                VerticalAlignment:=vertical_alignment.Center
                                                                Widget:=texture_block{DefaultImage:=Images.EmptyTexture,DefaultDesiredSize:=vector2{X:=500.0,Y:=185.0}}
                                                            overlay_slot:
                                                                HorizontalAlignment:=horizontal_alignment.Fill
                                                                VerticalAlignment:=vertical_alignment.Fill
                                                                Widget:=EquipPet_1_ChangeSlotButton  
                                stack_box_slot: 
                                    HorizontalAlignment:=horizontal_alignment.Center
                                    VerticalAlignment:=vertical_alignment.Center
                                    Padding:=RowPadding
                                    Widget:=stack_box:  #COLUMN 2
                                        Orientation:=orientation.Vertical
                                            Slots:=array:
                                                stack_box_slot: 
                                                    Padding:=RowPadding
                                                    VerticalAlignment:=vertical_alignment.Center
                                                    HorizontalAlignment:=horizontal_alignment.Center
                                                    Widget:=EquipPet_2_SlotInfo
                                                stack_box_slot: 
                                                    Padding:=margin{Right:=SuperLowPadding,Left:=SuperLowPadding}
                                                    HorizontalAlignment:=horizontal_alignment.Center
                                                    VerticalAlignment:=vertical_alignment.Center
                                                    Widget:=stack_box:  
                                                        Orientation:=orientation.Vertical
                                                        Slots:=array:
                                                            stack_box_slot: 
                                                                Padding:=margin{Top:=SelectorBoxPadding,Bottom:=SelectorBoxPadding}
                                                                Widget:=overlay:
                                                                    Slots:=array:
                                                                        overlay_slot:
                                                                            HorizontalAlignment:=horizontal_alignment.Fill
                                                                            VerticalAlignment:=vertical_alignment.Fill
                                                                            Widget:=texture_block{DefaultImage:=Images.Img_LineBackground}
                                                                        overlay_slot:
                                                                            HorizontalAlignment:=horizontal_alignment.Fill
                                                                            VerticalAlignment:=vertical_alignment.Fill
                                                                            Padding:=margin{Left:=SelectorBoxPadding,Right:=SelectorBoxPadding,Top:=SelectorBoxPadding,Bottom:=SelectorBoxPadding}
                                                                            Widget:=texture_block{DefaultImage:=Images.Img_LineBackground}
                                                                        overlay_slot:
                                                                            HorizontalAlignment:=horizontal_alignment.Center
                                                                            VerticalAlignment:=vertical_alignment.Center
                                                                            Widget:=texture_block{DefaultImage:=Images.EmptyTexture,DefaultDesiredSize:=vector2{X:=600.0,Y:=615.0}}
                                                                        overlay_slot:
                                                                            HorizontalAlignment:=horizontal_alignment.Center
                                                                            VerticalAlignment:=vertical_alignment.Center
                                                                            Widget:=stack_box:  
                                                                                Orientation:=orientation.Vertical
                                                                                Slots:=array:
                                                                                    stack_box_slot: 
                                                                                        Widget:=stack_box:  
                                                                                            Orientation:=orientation.Horizontal
                                                                                            Slots:=array:
                                                                                                stack_box_slot: 
                                                                                                    Padding:=RowPadding
                                                                                                    Widget:=EquipPet_2_TextureBlock_PetImages
                                                                                                stack_box_slot: 
                                                                                                    HorizontalAlignment:=horizontal_alignment.Center
                                                                                                    VerticalAlignment:=vertical_alignment.Center
                                                                                                    Widget:=stack_box:  
                                                                                                        Orientation:=orientation.Vertical
                                                                                                        Slots:=array:
                                                                                                            stack_box_slot: 
                                                                                                                HorizontalAlignment:=horizontal_alignment.Center
                                                                                                                VerticalAlignment:=vertical_alignment.Center
                                                                                                                #Padding:=margin{Left:=TextPetsBoxPadding}
                                                                                                                Widget:=stack_box:
                                                                                                                    Orientation:=orientation.Horizontal
                                                                                                                    Slots:=array:
                                                                                                                        stack_box_slot: 
                                                                                                                            Padding:=margin{Left:=2.5,Right:=2.5}
                                                                                                                            Widget:=EquipPet_2_TextBlock_Class
                                                                                                                        stack_box_slot: 
                                                                                                                            Padding:=margin{Left:=2.5, Right:=1.0}
                                                                                                                            Widget:=EquipPet_2_TextBlock_PackA
                                                                                                                        stack_box_slot: 
                                                                                                                            Padding:=margin{Right:=2.5}
                                                                                                                            Widget:=EquipPet_2_TextBlock_PackB
                                                                                                            stack_box_slot: 
                                                                                                                Padding:=RowPadding
                                                                                                                Widget:=EquipPet_2_TextBlock_Purity
                                                                                    stack_box_slot: 
                                                                                        HorizontalAlignment:=horizontal_alignment.Fill
                                                                                        VerticalAlignment:=vertical_alignment.Fill
                                                                                        Widget:=texture_block{DefaultImage:=Images.Img_HorizontalLine,DefaultDesiredSize:=vector2{Y:=3.0}}
                                                                                    stack_box_slot:  
                                                                                        HorizontalAlignment:=horizontal_alignment.Center
                                                                                        VerticalAlignment:=vertical_alignment.Center
                                                                                        Padding:=RowPadding
                                                                                        Widget:=EquipPet_2_Text_PackageA_LV
                                                                                    stack_box_slot: 
                                                                                        HorizontalAlignment:=horizontal_alignment.Fill
                                                                                        VerticalAlignment:=vertical_alignment.Fill
                                                                                        Widget:=texture_block{DefaultImage:=Images.Img_HorizontalLine,DefaultDesiredSize:=vector2{Y:=3.0}}
                                                                                    stack_box_slot: 
                                                                                        HorizontalAlignment:=horizontal_alignment.Fill
                                                                                        VerticalAlignment:=vertical_alignment.Fill
                                                                                        Widget:=stack_box:  
                                                                                            Orientation:=orientation.Horizontal
                                                                                            Slots:=array:
                                                                                                stack_box_slot: 
                                                                                                    Widget:=stack_box:
                                                                                                        Orientation:=orientation.Vertical
                                                                                                        Slots:=array:
                                                                                                stack_box_slot: 
                                                                                                    Widget:=stack_box:
                                                                                                        Orientation:=orientation.Vertical
                                                                                                        Slots:=array:
                                                                                                            stack_box_slot: 
                                                                                                                Padding:=RowPadding
                                                                                                                Widget:=text_block{DefaultText:=BasicMessage("Player enhancements"), DefaultTextColor:=Cyan, DefaultShadowOffset:=ShadowOffset, DefaultShadowOpacity := DefaultTextShadowOpacity}
                                                                                                            stack_box_slot: 
                                                                                                                HorizontalAlignment:=horizontal_alignment.Fill
                                                                                                                VerticalAlignment:=vertical_alignment.Fill
                                                                                                                Widget:=stack_box:  
                                                                                                                    Orientation:=orientation.Horizontal
                                                                                                                    Slots:=array:
                                                                                                                        stack_box_slot: 
                                                                                                                            HorizontalAlignment:=horizontal_alignment.Fill
                                                                                                                            VerticalAlignment:=vertical_alignment.Fill
                                                                                                                            Widget:=stack_box:  
                                                                                                                                Orientation:=orientation.Vertical
                                                                                                                                Slots:=array:
                                                                                                                                    stack_box_slot: 
                                                                                                                                        Padding:=RowPadding
                                                                                                                                        HorizontalAlignment:=horizontal_alignment.Left
                                                                                                                                        Widget:=EquipPet_2_TextBlock_Description_SpecialPerkA
                                                                                                                                    stack_box_slot: 
                                                                                                                                        Padding:=RowPadding
                                                                                                                                        HorizontalAlignment:=horizontal_alignment.Left
                                                                                                                                        Widget:=text_block{DefaultText:=BasicMessage("Pick damage:"), DefaultTextColor:=White, DefaultShadowOffset:=ShadowOffset, DefaultShadowOpacity := DefaultTextShadowOpacity}
                                                                                                                                    stack_box_slot: 
                                                                                                                                        Padding:=RowPadding
                                                                                                                                        HorizontalAlignment:=horizontal_alignment.Left
                                                                                                                                        Widget:=text_block{DefaultText:=BasicMessage("Weak point mult:"), DefaultTextColor:=White, DefaultShadowOffset:=ShadowOffset, DefaultShadowOpacity := DefaultTextShadowOpacity}
                                                                                                                        stack_box_slot: 
                                                                                                                            HorizontalAlignment:=horizontal_alignment.Fill
                                                                                                                            VerticalAlignment:=vertical_alignment.Fill
                                                                                                                            Widget:=stack_box:  
                                                                                                                                Orientation:=orientation.Vertical
                                                                                                                                Slots:=array:
                                                                                                                                    stack_box_slot: 
                                                                                                                                        Padding:=RowPadding
                                                                                                                                        Widget:=EquipPet_2_TextBlock_Current_SpecialPerkA
                                                                                                                                    stack_box_slot: 
                                                                                                                                        Padding:=RowPadding
                                                                                                                                        Widget:=EquipPet_2_TextBlock_Current_PlayerPickDamage
                                                                                                                                    stack_box_slot:
                                                                                                                                        Padding:=RowPadding 
                                                                                                                                        Widget:=EquipPet_2_TextBlock_Current_PlayerWeakPoint
                                                                                    stack_box_slot: 
                                                                                        HorizontalAlignment:=horizontal_alignment.Fill
                                                                                        VerticalAlignment:=vertical_alignment.Fill
                                                                                        Widget:=texture_block{DefaultImage:=Images.Img_HorizontalLine,DefaultDesiredSize:=vector2{Y:=3.0}}
                                                                                    stack_box_slot:  
                                                                                        HorizontalAlignment:=horizontal_alignment.Center
                                                                                        VerticalAlignment:=vertical_alignment.Center
                                                                                        Padding:=RowPadding
                                                                                        Widget:=EquipPet_2_Text_PackageB_LV
                                                                                    stack_box_slot: 
                                                                                        HorizontalAlignment:=horizontal_alignment.Fill
                                                                                        VerticalAlignment:=vertical_alignment.Fill
                                                                                        Widget:=texture_block{DefaultImage:=Images.Img_HorizontalLine,DefaultDesiredSize:=vector2{Y:=3.0}}
                                                                                    stack_box_slot: 
                                                                                        HorizontalAlignment:=horizontal_alignment.Fill
                                                                                        VerticalAlignment:=vertical_alignment.Fill
                                                                                        Widget:=stack_box:  
                                                                                            Orientation:=orientation.Horizontal
                                                                                            Slots:=array:
                                                                                                stack_box_slot: 
                                                                                                    Widget:=stack_box:
                                                                                                        Orientation:=orientation.Vertical
                                                                                                        Slots:=array:
                                                                                                            stack_box_slot: 
                                                                                                                Padding:=RowPadding
                                                                                                                Widget:=text_block{DefaultText:=BasicMessage("Player enhancements"), DefaultTextColor:=Cyan, DefaultShadowOffset:=ShadowOffset, DefaultShadowOpacity := DefaultTextShadowOpacity}
                                                                                                            stack_box_slot: 
                                                                                                                HorizontalAlignment:=horizontal_alignment.Fill
                                                                                                                VerticalAlignment:=vertical_alignment.Fill
                                                                                                                Widget:=stack_box:  
                                                                                                                    Orientation:=orientation.Horizontal
                                                                                                                    Slots:=array:
                                                                                                                        stack_box_slot: 
                                                                                                                            HorizontalAlignment:=horizontal_alignment.Fill
                                                                                                                            VerticalAlignment:=vertical_alignment.Fill
                                                                                                                            Widget:=stack_box:  
                                                                                                                                Orientation:=orientation.Vertical
                                                                                                                                Slots:=array:
                                                                                                                                    stack_box_slot: 
                                                                                                                                        Padding:=RowPadding
                                                                                                                                        HorizontalAlignment:=horizontal_alignment.Left
                                                                                                                                        Widget:=EquipPet_2_TextBlock_Description_SpecialPerkB
                                                                                                                                    stack_box_slot: 
                                                                                                                                        Padding:=RowPadding
                                                                                                                                        HorizontalAlignment:=horizontal_alignment.Left
                                                                                                                                        Widget:=text_block{DefaultText:=BasicMessage("Pick efficiency:"), DefaultTextColor:=White, DefaultShadowOffset:=ShadowOffset, DefaultShadowOpacity := DefaultTextShadowOpacity}
                                                                                                                                    stack_box_slot: 
                                                                                                                                        Padding:=RowPadding
                                                                                                                                        HorizontalAlignment:=horizontal_alignment.Left
                                                                                                                                        Widget:=text_block{DefaultText:=BasicMessage("Pick critical dam:"), DefaultTextColor:=White, DefaultShadowOffset:=ShadowOffset, DefaultShadowOpacity := DefaultTextShadowOpacity}
                                                                                                                        stack_box_slot: 
                                                                                                                            HorizontalAlignment:=horizontal_alignment.Fill
                                                                                                                            VerticalAlignment:=vertical_alignment.Fill
                                                                                                                            Widget:=stack_box:  
                                                                                                                                Orientation:=orientation.Vertical
                                                                                                                                Slots:=array:
                                                                                                                                    stack_box_slot: 
                                                                                                                                        Padding:=RowPadding
                                                                                                                                        Widget:=EquipPet_2_TextBlock_Current_SpecialPerkB
                                                                                                                                    stack_box_slot: 
                                                                                                                                        Padding:=RowPadding
                                                                                                                                        Widget:=EquipPet_2_TextBlock_Current_PlayerPickEfficiency
                                                                                                                                    stack_box_slot: 
                                                                                                                                        Padding:=RowPadding
                                                                                                                                        Widget:=EquipPet_2_TextBlock_Current_PlayerPickCriticalDamage   
                                                stack_box_slot:
                                                    Padding:=RowPadding
                                                    VerticalAlignment:=vertical_alignment.Center
                                                    HorizontalAlignment:=horizontal_alignment.Center
                                                    Widget:=overlay:
                                                        Slots:=array:
                                                            overlay_slot:
                                                                HorizontalAlignment:=horizontal_alignment.Center
                                                                VerticalAlignment:=vertical_alignment.Center
                                                                Widget:=texture_block{DefaultImage:=Images.EmptyTexture,DefaultDesiredSize:=vector2{X:=500.0,Y:=185.0}}
                                                            overlay_slot:
                                                                HorizontalAlignment:=horizontal_alignment.Fill
                                                                VerticalAlignment:=vertical_alignment.Fill
                                                                Widget:=EquipPet_2_ChangeSlotButton                                                                        
                canvas_slot:
                    Anchors:=anchors{Minimum:=vector2{X:=0.5,Y:=0.5},Maximum:=vector2{X:=0.5,Y:=0.5}}
                    ZOrder:=0
                    SizeToContent:=true
                    Offsets:=margin{Left:=0.0,Top:=480.0,Right:=0.0,Bottom:=0.0} #PositionX, PositionY, SizeX, SizeY
                    Alignment:=vector2{X:=0.5,Y:=0.5}
                    Widget:=overlay:             
                        Slots:=array:
                            overlay_slot:
                                Widget:=texture_block{DefaultImage:=Images.EmptyTexture,DefaultDesiredSize:=vector2{X:=BigCloseButtonSizeX,Y:=80.0}}
                            overlay_slot: 
                                HorizontalAlignment:=horizontal_alignment.Fill
                                VerticalAlignment:=vertical_alignment.Fill
                                Widget:=EquipPet_CloseButton

    EquipPet_RefreshAll() : void =
        if(Agent := MaybeAgent?, Player:=player[Agent], Player.IsActive[]):
            MaybeBakery:=GameManager.GetAgentBakery(Agent)
            if(ValidBakery := MaybeBakery?):
                EquippedPetsTupple := GetEquippedPets(ValidBakery.PlayerDataCache)
                set AllEquippedPetsIndexes=EquippedPetsTupple(1)
                set AllEquippedPetTables=EquippedPetsTupple(0)
                loop:
                    if(AllEquippedPetsIndexes.Length<3):
                        set AllEquippedPetsIndexes = AllEquippedPetsIndexes + array{-1}
                    else:
                        break
                #Clear
                for(Index:=0..AllEquipPet_SlotInfo.Length-1):
                    if:
                        EquipPet_SlotInfo:= AllEquipPet_SlotInfo[Index]
                        EquipPet_ChangeSlotButton := AllEquipPet_ChangeSlot[Index]
                        EquipPet_Text_PackageA_LV:=  AllEquipPet_Text_PackageA_LV[Index]
                        EquipPet_Text_PackageB_LV := AllEquipPet_Text_PackageB_LV[Index]
                        #TOURNAMENT
                        EquipPet_TextureBlock_PetImages := AllEquipPet_TextureBlock_PetImages[Index]
                        EquipPet_TextBlock_Purity := AllEquipPet_TextBlock_Purity[Index]
                        EquipPet_TextBlock_Class := AllEquipPet_TextBlock_Class[Index]
                        EquipPet_TextBlock_PackA := AllEquipPet_TextBlock_PackA[Index]
                        EquipPet_TextBlock_PackB := AllEquipPet_TextBlock_PackB[Index]
                        #Package A
                        EquipPet_TextBlock_Description_SpecialPerkA := AllEquipPet_TextBlock_Description_SpecialPerkA[Index]
                        EquipPet_TextBlock_Current_SpecialPerkA:= AllEquipPet_TextBlock_Current_SpecialPerkA[Index]
                        EquipPet_TextBlock_Current_PlayerPickDamage := AllEquipPet_TextBlock_Current_PlayerPickDamage[Index]
                        EquipPet_TextBlock_Current_PlayerWeakPoint := AllEquipPet_TextBlock_Current_PlayerWeakPoint[Index]
                        #Package B
                        EquipPet_TextBlock_Description_SpecialPerkB := AllEquipPet_TextBlock_Description_SpecialPerkB[Index]
                        EquipPet_TextBlock_Current_SpecialPerkB := AllEquipPet_TextBlock_Current_SpecialPerkB[Index]
                        EquipPet_TextBlock_Current_PlayerPickEfficiency := AllEquipPet_TextBlock_Current_PlayerPickEfficiency[Index]
                        EquipPet_TextBlock_Current_PlayerPickCriticalDamage := AllEquipPet_TextBlock_Current_PlayerPickCriticalDamage[Index]
                    then:
                        if(Index<=ValidBakery.PlayerDataCache.MaxActive_PetSlots-1):
                            EquipPet_SlotInfo.SetTextColor(Lime)
                            EquipPet_SlotInfo.SetText(BasicMessage("SLOT AVAILABLE"))
                            EquipPet_ChangeSlotButton.SetText(BasicMessage("SLOT AVAILABLE"))
                        else:
                            EquipPet_SlotInfo.SetTextColor(Red)
                            EquipPet_SlotInfo.SetText(BasicMessage("SLOT LOCKED"))
                            EquipPet_ChangeSlotButton.SetText(BasicMessage("SLOT LOCKED"))
                        RefreshPetStats(
                            -1,
                            array{},
                            false,
                            EquipPet_TextureBlock_PetImages,

                            EquipPet_Text_PackageA_LV,
                            EquipPet_Text_PackageB_LV,
                            
                            EquipPet_TextBlock_Current_SpecialPerkA,
                            EquipPet_TextBlock_Current_SpecialPerkB,
                            EquipPet_TextBlock_Description_SpecialPerkA,
                            EquipPet_TextBlock_Description_SpecialPerkB,
                            EquipPet_TextBlock_Purity,
                            EquipPet_TextBlock_Class,
                            EquipPet_TextBlock_PackA,
                            EquipPet_TextBlock_PackB,
                            
                            Empty_TextBlock,
                            Empty_TextBlock,
                            Empty_TextBlock,

                            EquipPet_TextBlock_Current_PlayerPickDamage,
                            EquipPet_TextBlock_Current_PlayerWeakPoint,

                            Empty_TextBlock,
                            Empty_TextBlock,
                            Empty_TextBlock,

                            EquipPet_TextBlock_Current_PlayerPickEfficiency,
                            EquipPet_TextBlock_Current_PlayerPickCriticalDamage,

                            array{},
                            array{},
                            array{},
                            array{},
                            array{},
                            array{})
                    else:
                        PrintFatalError("EquipPet_RefreshAll: No Index {Index} found")
                #SetAllUsedPets
                for(Index:=0..AllEquippedPetTables.Length-1,EquippedPet:=AllEquippedPetTables[Index]):
                    if:
                        EquipPet_SlotInfo:= AllEquipPet_SlotInfo[Index]
                        EquipPet_ChangeSlotButton := AllEquipPet_ChangeSlot[Index]
                        EquipPet_Text_PackageA_LV:=  AllEquipPet_Text_PackageA_LV[Index]
                        EquipPet_Text_PackageB_LV := AllEquipPet_Text_PackageB_LV[Index]
                        #TOURNAMENT
                        EquipPet_TextureBlock_PetImages := AllEquipPet_TextureBlock_PetImages[Index]
                        EquipPet_TextBlock_Purity := AllEquipPet_TextBlock_Purity[Index]
                        EquipPet_TextBlock_Class := AllEquipPet_TextBlock_Class[Index]
                        EquipPet_TextBlock_PackA := AllEquipPet_TextBlock_PackA[Index]
                        EquipPet_TextBlock_PackB := AllEquipPet_TextBlock_PackB[Index]
                        #Package A
                        EquipPet_TextBlock_Description_SpecialPerkA := AllEquipPet_TextBlock_Description_SpecialPerkA[Index]
                        EquipPet_TextBlock_Current_SpecialPerkA:= AllEquipPet_TextBlock_Current_SpecialPerkA[Index]
                        EquipPet_TextBlock_Current_PlayerPickDamage := AllEquipPet_TextBlock_Current_PlayerPickDamage[Index]
                        EquipPet_TextBlock_Current_PlayerWeakPoint := AllEquipPet_TextBlock_Current_PlayerWeakPoint[Index]
                        #Package B
                        EquipPet_TextBlock_Description_SpecialPerkB := AllEquipPet_TextBlock_Description_SpecialPerkB[Index]
                        EquipPet_TextBlock_Current_SpecialPerkB := AllEquipPet_TextBlock_Current_SpecialPerkB[Index]
                        EquipPet_TextBlock_Current_PlayerPickEfficiency := AllEquipPet_TextBlock_Current_PlayerPickEfficiency[Index]
                        EquipPet_TextBlock_Current_PlayerPickCriticalDamage := AllEquipPet_TextBlock_Current_PlayerPickCriticalDamage[Index]
                    then:
                        EquipPet_ChangeSlotButton.SetText(BasicMessage("CHANGE"))
                        EquipPet_SlotInfo.SetText(BasicMessage("SLOT USED"))
                        EquipPet_SlotInfo.SetTextColor(Coral)
                        RefreshPetStats(
                            -1,
                            array{},
                            option{EquippedPet},
                            EquipPet_TextureBlock_PetImages,

                            EquipPet_Text_PackageA_LV,
                            EquipPet_Text_PackageB_LV,
                            
                            EquipPet_TextBlock_Current_SpecialPerkA,
                            EquipPet_TextBlock_Current_SpecialPerkB,
                            EquipPet_TextBlock_Description_SpecialPerkA,
                            EquipPet_TextBlock_Description_SpecialPerkB,
                            EquipPet_TextBlock_Purity,
                            EquipPet_TextBlock_Class,
                            EquipPet_TextBlock_PackA,
                            EquipPet_TextBlock_PackB,
                            
                            
                            Empty_TextBlock,
                            Empty_TextBlock,
                            Empty_TextBlock,

                            EquipPet_TextBlock_Current_PlayerPickDamage,
                            EquipPet_TextBlock_Current_PlayerWeakPoint,

                            Empty_TextBlock,
                            Empty_TextBlock,
                            Empty_TextBlock,

                            EquipPet_TextBlock_Current_PlayerPickEfficiency,
                            EquipPet_TextBlock_Current_PlayerPickCriticalDamage,

                            array{},
                            array{},
                            array{},
                            array{},
                            array{},
                            array{})


    
    EquipPet_Remove_Canvas(Message : widget_message) : void =
        set EquipPet_UIIsDisplayed = false
        if:
            Agent := MaybeAgent?
            Player:=player[Agent]
            PlayerUI:=GetPlayerUI[Player]
            Canvas := EquipPet_MaybeCanvas?
        then:
            #Print("Removing EquipPet_Remove_Canvas")
            PlayerUI.RemoveWidget(Canvas)
            EquipPet_RemoveDelay_OnRemove()

    EquipPet_Remove_CanvasWithoutMessage() : void =
        set EquipPet_UIIsDisplayed = false
        if:
            Agent := MaybeAgent?
            Player:=player[Agent]
            PlayerUI:=GetPlayerUI[Player]
            Canvas := EquipPet_MaybeCanvas?
        then:
            #Print("Removing EquipPet_Remove_Canvas")
            PlayerUI.RemoveWidget(Canvas)
            EquipPet_RemoveDelay_OnRemove()


    EquipPet_RemoveDelay_OnAdd() : void =
        EquipPet_CloseButton.SetText(BasicMessage("Close"))

    EquipPet_RemoveDelay_OnRemove() : void =
        EquipPet_CloseButton.SetText(BasicMessage("Close "))
    
    EquipPet_Display_Canvas() : void =
        #Print("EquipPet_Display_Canvas")
        if(not EquipPet_UIIsDisplayed?):
            set EquipPet_UIIsDisplayed = true
            var Index : int = 0 
            if:
                Agent := MaybeAgent?
                PlayerUI:=GetPlayerUI[player[Agent]]
                Canvas := EquipPet_MaybeCanvas?
            then:
                PlayerUI.AddWidget(Canvas,player_ui_slot{ZOrder:=0, InputMode := ui_input_mode.All})
                EquipPet_RemoveDelay_OnAdd()
                EquipPet_RefreshAll()

    ################
    #MAIN MENU CANVAS
    MakeMainMenuPetsCanvas(): canvas=
        Canvas := canvas:
            Slots:=array:
                #Main overlay
                canvas_slot:
                    Anchors:=anchors{Minimum:=vector2{X:=0.5,Y:=0.5},Maximum:=vector2{X:=0.5,Y:=0.5}}
                    ZOrder:=0
                    SizeToContent:=true
                    Offsets:=margin{Left:=0.0,Top:=0.0,Right:=0.0,Bottom:=0.0} #PositionX, PositionY, SizeX, SizeY
                    Alignment:=vector2{X:=0.5,Y:=0.5}
                    Widget:=texture_block{DefaultImage:=Images.BlackBackground,DefaultDesiredSize:=vector2{X:=3000.0,Y:=1500.0}}
                canvas_slot:
                    Anchors:=anchors{Minimum:=vector2{X:=0.5,Y:=0.5},Maximum:=vector2{X:=0.5,Y:=0.5}}
                    ZOrder:=1
                    SizeToContent:=true
                    Offsets:=margin{Left:=0.0,Top:=0.0,Right:=0.0,Bottom:=0.0} #PositionX, PositionY, SizeX, SizeY
                    Alignment:=vector2{X:=0.5,Y:=0.5}
                    Widget:=texture_block{DefaultImage:=Images.Img_MainBackground,DefaultDesiredSize:=vector2{X:=850.0,Y:=1250.0}}
                canvas_slot:
                    Anchors:=anchors{Minimum:=vector2{X:=0.5,Y:=0.5},Maximum:=vector2{X:=0.5,Y:=0.5}}
                    ZOrder:=2
                    SizeToContent:=true
                    Offsets:=margin{Left:=0.0,Top:=0.0,Right:=0.0,Bottom:=0.0} #PositionX, PositionY, SizeX, SizeY
                    Alignment:=vector2{X:=0.5,Y:=0.5}
                    Widget:=stack_box:  
                        Orientation:=orientation.Vertical
                            Slots:=array:
                                stack_box_slot:
                                    Padding:=margin{Top:=5.0,Bottom:=5.0,Left:=5.0,Right:=5.0}
                                    Widget:=texture_block{DefaultImage:=Images.Icon_ADC,DefaultDesiredSize:=vector2{X:=200.0,Y:=200.0}}
                                stack_box_slot:
                                    HorizontalAlignment:=horizontal_alignment.Fill
                                    Padding:=margin{Top:=5.0,Bottom:=5.0,Left:=5.0,Right:=5.0}
                                    Widget:=overlay:
                                        Slots:=array:
                                            overlay_slot:
                                                Widget:=texture_block{DefaultImage:=Images.EmptyTexture,DefaultDesiredSize:=vector2{X:=610.0,Y:=80.0}}
                                            overlay_slot:
                                                HorizontalAlignment:=horizontal_alignment.Fill
                                                VerticalAlignment:=vertical_alignment.Fill
                                                Widget:=MainMenu_CreatePetButton
                                stack_box_slot:
                                    HorizontalAlignment:=horizontal_alignment.Fill
                                    Padding:=margin{Top:=5.0,Bottom:=5.0,Left:=5.0,Right:=5.0}
                                    Widget:=overlay:
                                        Slots:=array:
                                            overlay_slot:
                                                Widget:=texture_block{DefaultImage:=Images.EmptyTexture,DefaultDesiredSize:=vector2{X:=610.0,Y:=80.0}}
                                            overlay_slot:
                                                HorizontalAlignment:=horizontal_alignment.Fill
                                                VerticalAlignment:=vertical_alignment.Fill
                                                Widget:=MainMenu_EquipPetButton
                                stack_box_slot:
                                    HorizontalAlignment:=horizontal_alignment.Fill
                                    Padding:=margin{Top:=5.0,Bottom:=5.0,Left:=5.0,Right:=5.0}
                                    Widget:=overlay:
                                        Slots:=array:
                                            overlay_slot:
                                                Widget:=texture_block{DefaultImage:=Images.EmptyTexture,DefaultDesiredSize:=vector2{X:=610.0,Y:=80.0}}
                                            overlay_slot:
                                                HorizontalAlignment:=horizontal_alignment.Fill
                                                VerticalAlignment:=vertical_alignment.Fill
                                                Widget:=MainMenu_ImprovePetButton
                                stack_box_slot:
                                    HorizontalAlignment:=horizontal_alignment.Fill
                                    Padding:=margin{Top:=5.0,Bottom:=5.0,Left:=5.0,Right:=5.0}
                                    Widget:=overlay:
                                        Slots:=array:
                                            overlay_slot:
                                                Widget:=texture_block{DefaultImage:=Images.EmptyTexture,DefaultDesiredSize:=vector2{X:=610.0,Y:=80.0}}
                                            overlay_slot:
                                                HorizontalAlignment:=horizontal_alignment.Fill
                                                VerticalAlignment:=vertical_alignment.Fill
                                                Widget:=MainMenu_BuySlotsPetButton
                                stack_box_slot:
                                    HorizontalAlignment:=horizontal_alignment.Fill
                                    Padding:=margin{Top:=5.0,Bottom:=5.0,Left:=5.0,Right:=5.0}
                                    Widget:=overlay:
                                        Slots:=array:
                                            overlay_slot:
                                                Widget:=texture_block{DefaultImage:=Images.EmptyTexture,DefaultDesiredSize:=vector2{X:=610.0,Y:=80.0}}
                                            overlay_slot:
                                                HorizontalAlignment:=horizontal_alignment.Fill
                                                VerticalAlignment:=vertical_alignment.Fill
                                                Widget:=MainMenu_RecyclePetButton
                                stack_box_slot:
                                    HorizontalAlignment:=horizontal_alignment.Fill
                                    Padding:=margin{Top:=5.0,Bottom:=5.0,Left:=5.0,Right:=5.0}
                                    Widget:=overlay:
                                        Slots:=array:
                                            overlay_slot:
                                                Widget:=texture_block{DefaultImage:=Images.EmptyTexture,DefaultDesiredSize:=vector2{X:=610.0,Y:=80.0}}
                                            overlay_slot:
                                                HorizontalAlignment:=horizontal_alignment.Fill
                                                VerticalAlignment:=vertical_alignment.Fill
                                                Widget:=MainMenu_MixPetsButton
                                stack_box_slot:
                                    HorizontalAlignment:=horizontal_alignment.Fill
                                    Padding:=margin{Top:=5.0,Bottom:=5.0,Left:=5.0,Right:=5.0}
                                    Widget:=overlay:
                                        Slots:=array:
                                            overlay_slot:
                                                Widget:=texture_block{DefaultImage:=Images.EmptyTexture,DefaultDesiredSize:=vector2{X:=610.0,Y:=80.0}}
                                            overlay_slot:
                                                HorizontalAlignment:=horizontal_alignment.Fill
                                                VerticalAlignment:=vertical_alignment.Fill
                                                Widget:=MainMenu_CloseButton


    Handler_FromEverything_To_Menu(Message : widget_message, SubMenuOption : int)<suspends> : void =
        if(Agent  := agent[Message.Player], ValidBakery:=GameManager.GetAgentBakery_Transacted(Agent)?):
            if(SubMenuOption=SUBMENU_CREATE_PET):
                CreatePet_Remove_Canvas(Message)
                Sleep(0.1)
                LoadEquippedPets(ValidBakery.PlayerDataCache)
                MainMenu_Display_Canvas()
            else if(SubMenuOption=SUBMENU_EQUIP_PET):
                EquipPet_Remove_Canvas(Message)
                Sleep(0.1)
                LoadEquippedPets(ValidBakery.PlayerDataCache)
                MainMenu_Display_Canvas()
            else if(SubMenuOption=SUBMENU_IMPROVE_PET, PetsUpgradesUI:=MaybePetsUpgradesUI?):
                PetsUpgradesUI.Remove_Canvas(Message)
                Sleep(0.1)
                LoadEquippedPets(ValidBakery.PlayerDataCache)
                MainMenu_Display_Canvas()
            else if(SubMenuOption=SUBMENU_RECYCLE_PET):
                RecyclePet_Remove_Canvas(Message)
                Sleep(0.1)
                LoadEquippedPets(ValidBakery.PlayerDataCache)
                MainMenu_Display_Canvas()
            else if(SubMenuOption=SUBMENU_BUY_SLOTSPET):
                BuySlots_Remove_Canvas(Message)
                Sleep(0.1)
                LoadEquippedPets(ValidBakery.PlayerDataCache)
                MainMenu_Display_Canvas()
            else if(SubMenuOption=SUBMENU_MIX_PET):
                MixPets_Remove_Canvas(Message)
                Sleep(0.1)
                LoadEquippedPets(ValidBakery.PlayerDataCache)
                MainMenu_Display_Canvas()
        else:
            CreatePet_Remove_Canvas(Message)
            EquipPet_Remove_Canvas(Message)
            if(PetsUpgradesUI:=MaybePetsUpgradesUI?):
                PetsUpgradesUI.Remove_Canvas(Message)
            RecyclePet_Remove_Canvas(Message)
            BuySlots_Remove_Canvas(Message)
            MixPets_Remove_Canvas(Message)
            PrintFatalError("On Handler_FromEverything_To_Menu: Agent Bakery not found!")

    Handler_FromMainMenu_To_Everything(Message : widget_message, SubMenuOption : int)<suspends> : void =
        if(Agent:=MaybeAgent?):
            MaybeBakery:=GameManager.GetAgentBakery(Agent)
                if(ValidBakery := MaybeBakery?, Message.Player.IsActive[]):
                    if(SubMenuOption=SUBMENU_CREATE_PET):
                        MainMenu_Remove_Canvas(Message)
                        Sleep(0.1)
                        CreatePet_Display_Canvas()
                    else if(SubMenuOption=SUBMENU_EQUIP_PET):
                        if(ValidBakery.PlayerDataCache.AllPets.Length>0):
                            MainMenu_Remove_Canvas(Message)
                            Sleep(0.1)
                            EquipPet_Display_Canvas()
                        else:
                            RunMenuError(MainMenu_EquipPetButton, "Not sourdough found", "Change sourdough slots")
                    else if(SubMenuOption=SUBMENU_IMPROVE_PET, PetsUpgradesUI:=MaybePetsUpgradesUI?):
                        if(ValidBakery.PlayerDataCache.AllPets.Length>0):
                            MainMenu_Remove_Canvas(Message)
                            Sleep(0.1)
                            PetsUpgradesUI.Display_Canvas()
                        else:
                            RunMenuError(MainMenu_ImprovePetButton, "Not sourdough found", "Improve sourdough")
                    else if(SubMenuOption=SUBMENU_RECYCLE_PET):
                        if(ValidBakery.PlayerDataCache.AllPets.Length>0):
                            MainMenu_Remove_Canvas(Message)
                            Sleep(0.1)
                            RecyclePet_Display_Canvas()
                        else:
                            RunMenuError(MainMenu_RecyclePetButton, "Not sourdough found", "Recycle sourdough")
                    else if(SubMenuOption=SUBMENU_BUY_SLOTSPET):
                        MainMenu_Remove_Canvas(Message)
                        Sleep(0.1)
                        BuySlots_Display_Canvas()
                    else if(SubMenuOption=SUBMENU_MIX_PET):
                        if(ValidBakery.PlayerDataCache.AllPets.Length>0):
                            MainMenu_Remove_Canvas(Message)
                            Sleep(0.1)
                            MixPets_Display_Canvas()
                        else:
                            RunMenuError(MainMenu_MixPetsButton, "Not sourdough found", "Mix sourdoughs")
        
    RunMenuError(Button : button_regular, String : string, OldString : string)<suspends> : void =
        Button.SetText(BasicMessage(String))
        Sleep(1.0)
        Button.SetText(BasicMessage(OldString))
        
    #Main functions

    CloseAllMainMenu(Message : widget_message) : void =
        #ONLY USE WHEN CLOSE MENU BUTTON IS PRESSES
        set GeneralSelector_UIIsDisplayed = false
        set MainMenu_UIIsDisplayed = false
        set MixPets_UIIsDisplayed = false
        set EquipPet_UIIsDisplayed = false
        set RecyclePet_UIIsDisplayed = false
        set CreatePet_UIIsDisplayed = false
        set BuySlots_UIIsDisplayed = false
        MainMenu_Remove_Canvas(Message)

    MainMenu_Remove_Canvas(Message : widget_message) : void =
        set MainMenu_UIIsDisplayed = false
        if:
            Agent := MaybeAgent?
            Player:=player[Agent]
            PlayerUI:=GetPlayerUI[Player]
            Canvas := MainMenu_MaybeCanvas?
        then:
            #Print("Removing MainMenu_Remove_Canvas")
            PlayerUI.RemoveWidget(Canvas)
            MainMenu_RemoveDelay_OnRemove()
        Generate_PlayerEnhacements()
    MainMenu_Remove_CanvasWithoutMessage() : void =
        set MainMenu_UIIsDisplayed = false
        if:
            Agent := MaybeAgent?
            Player:=player[Agent]
            PlayerUI:=GetPlayerUI[Player]
            Canvas := MainMenu_MaybeCanvas?
        then:
            #Print("Removing MainMenu_Remove_Canvas")
            PlayerUI.RemoveWidget(Canvas)
            MainMenu_RemoveDelay_OnRemove()
        Generate_PlayerEnhacements()

    MainMenu_RemoveDelay_OnAdd() : void =
        MainMenu_CloseButton.SetText(BasicMessage("Close"))

    MainMenu_RemoveDelay_OnRemove() : void =
        MainMenu_CloseButton.SetText(BasicMessage("Close   "))
    
    MainMenu_Display_Canvas() : void =
        if(not MainMenu_UIIsDisplayed?):
            #Print("MainMenu_Display_Canvas")
            set MainMenu_UIIsDisplayed = true
            var Index : int = 0 
            if:
                Agent := MaybeAgent?
                PlayerUI:=GetPlayerUI[player[Agent]]
                Canvas := MainMenu_MaybeCanvas?
            then:
                PlayerUI.AddWidget(Canvas,player_ui_slot{ZOrder:=0, InputMode := ui_input_mode.All})
                MainMenu_RemoveDelay_OnAdd()
                #GeneralSelector_RefreshAll()

    #GENERAL SELECTOR
    ####################
    Handler_FromGeneralSelectorMenu_To_Everything(Message : widget_message, IsCancel : logic)<suspends> : void =
        GeneralSelector_Remove_Canvas(Message)
        Sleep(0.1)        
        set GeneralSelector_IsCancel = IsCancel
        if(GeneralSelector_OpenedFor = From_EquipPetsUI):
        else if(GeneralSelector_OpenedFor = From_MixPetsUI):
            MixPets_Display_Canvas()

    GeneralSelector_RefreshAll() : void =
        if(Agent := MaybeAgent?, Player := player[Agent], Player.IsActive[]):
            MaybeBakery:=GameManager.GetAgentBakery(Agent)
                if(ValidBakery := MaybeBakery?):
                    AllPets:=ValidBakery.PlayerDataCache.AllPets
                    if(Pet := AllPets[GeneralSelector_CurrentOption]):
                        RefreshPetStats(
                            GeneralSelector_CurrentOption,
                            ValidBakery.PlayerDataCache.AllPets,
                            option{Pet},
                            GeneralSelector_Selected_TextureBlock_PetImages,

                            GeneralSelector_Text_PackageA_LV,
                            GeneralSelector_Text_PackageB_LV,
                            
                            GeneralSelector_Selected_TextBlock_Current_SpecialPerkA,
                            GeneralSelector_Selected_TextBlock_Current_SpecialPerkB,
                            GeneralSelector_Selected_TextBlock_Description_SpecialPerkA,
                            GeneralSelector_Selected_TextBlock_Description_SpecialPerkB,
                            GeneralSelector_Selected_TextBlock_Purity,
                            GeneralSelector_Selected_TextBlock_Class,
                            GeneralSelector_Selected_TextBlock_PackA,
                            GeneralSelector_Selected_TextBlock_PackB,
                            
                            GeneralSelector_Selected_TextBlock_Current_TournamentHealth,
                            GeneralSelector_Selected_TextBlock_Current_TournamentDamage,
                            GeneralSelector_Selected_TextBlock_Current_TournamentSpeed,

                            GeneralSelector_Selected_TextBlock_Current_PlayerPickDamage,
                            GeneralSelector_Selected_TextBlock_Current_PlayerWeakPoint,

                            GeneralSelector_Selected_TextBlock_Current_TournamentActionSpeed,
                            GeneralSelector_Selected_TextBlock_Current_TournamentHealingPerk,
                            GeneralSelector_Selected_TextBlock_Current_TournamentRange,

                            GeneralSelector_Selected_TextBlock_Current_PlayerPickEfficiency,
                            GeneralSelector_Selected_TextBlock_Current_PlayerPickCriticalDamage,

                            AllGeneralSelector_TextureBlock_PetImages,
                            AllGeneralSelector_TextBlock_Classes,
                            AllGeneralSelector_TextBlock_PackA,
                            AllGeneralSelector_TextBlock_PackB,
                            AllGeneralSelector_TextBlock_Purity,
                            AllGeneralSelector_TextBlock_Tier
                            )
                    
                        if(GeneralSelector_FromSubmenu=SUBMENU_EQUIP_PET):
                            if(Pet.IsEquipped=true):
                                GeneralSelector_Selected_TextureBlock_PetImages.SetTint(Red)
                            else:
                                GeneralSelector_Selected_TextureBlock_PetImages.SetTint(White)

                        if(GeneralSelector_FromSubmenu=SUBMENU_MIX_PET):
                            var IsGoingToBeMixed : logic = false
                            for(IndexPetToMix:AllMixPets_PetsToMix, IndexPetToMix=GeneralSelector_CurrentOption):
                                set IsGoingToBeMixed = true
                            if(IsGoingToBeMixed?):
                                GeneralSelector_Selected_TextureBlock_PetImages.SetTint(Red)
                            else:
                                GeneralSelector_Selected_TextureBlock_PetImages.SetTint(White)
                                
                            
                    #The other images
                    
                    for(Index:=0..AllGeneralSelector_TextureBlock_PetImages.Length-1):
                        if:
                            TextureBlock_PetImage:=AllGeneralSelector_TextureBlock_PetImages[Index]
                            #TextBlock_Class:=AllGeneralSelector_TextBlock_Classes[Index]
                            #TextBlock_Purity:=AllGeneralSelector_TextBlock_Purity[Index]
                            #TextBlock_Tier:=AllGeneralSelector_TextBlock_Tier[Index]
                        then:
                            
                            if(PetIndexSelected:=AllPets[GeneralSelector_CurrentOption-2+Index]):
                                TuplePurity:=ConvertRawPurity(PetIndexSelected.Purity)
                                #SetPetImage(PetIndexSelected.Type, TextureBlock_PetImage)
                                #SetPetClass(PetIndexSelected.Type, TextBlock_Class)
                                #SetPetPurity(TuplePurity(0), TextBlock_Purity)
                                #SetPetTier(TuplePurity(1), TextBlock_Tier)

                                if(GeneralSelector_FromSubmenu=SUBMENU_MIX_PET):
                                    var IsGoingToBeMixed : logic = false 
                                    for(IndexPetToMix:AllMixPets_PetsToMix, IndexPetToMix=(GeneralSelector_CurrentOption-2+Index)):
                                        set IsGoingToBeMixed = true
                                    if(IsGoingToBeMixed?):
                                        TextureBlock_PetImage.SetTint(Red)
                                    else:
                                        TextureBlock_PetImage.SetTint(White)

                                if(GeneralSelector_FromSubmenu=SUBMENU_EQUIP_PET):
                                    if(PetIndexSelected.IsEquipped=true):
                                        TextureBlock_PetImage.SetTint(Red)
                                    else:
                                        TextureBlock_PetImage.SetTint(White)
                            else:
                                TextureBlock_PetImage.SetImage(Images.Cross_Icon)
                                TextureBlock_PetImage.SetTint(White)

                        


    HandlerGeneralSelector_SelectPet(Message : widget_message) : void =
        spawn:
            GeneralSelector_SelectPet(Message)

    GeneralSelector_SelectPet(Message : widget_message)<suspends> : void =
        GeneralSelector_Remove_CanvasWithoutMessage()
        Sleep(0.04)
        if(Agent := MaybeAgent?, Player := player[Agent], Player.IsActive[]):
            MaybeBakery:=GameManager.GetAgentBakery(Agent)
                if(ValidBakery := MaybeBakery?):
                    var AllPets : []pet_data_table = ValidBakery.PlayerDataCache.AllPets
                    #SUBMENU EQUIP NEW PET
                    if(GeneralSelector_FromSubmenu=SUBMENU_EQUIP_PET):
                        if(PetSlotIndex:=AllEquippedPetsIndexes[GeneralSelector_UISlotSelected], PetSlotIndex <> -1):
                            #Print("Pet of InUISlotSelect: {GeneralSelector_UISlotSelected}: `{PetSlotIndex}` is going to be replaced by `{GeneralSelector_CurrentOption}`")
                            if(PetTable:=AllPets[PetSlotIndex]):
                                UpdatedPet:=GameManager.PlayersDataManager.SetEquippedPet(PetTable, false) #Lets put this pet to false
                                if:
                                    set AllPets[PetSlotIndex] = UpdatedPet #And lets save if after that
                                    NewPetToUseNow := AllPets[GeneralSelector_CurrentOption]
                                    UpdatedNewPet := GameManager.PlayersDataManager.SetEquippedPet(NewPetToUseNow, true)
                                    set AllPets[GeneralSelector_CurrentOption] = UpdatedNewPet #Saved the new pet in all pet like being in use
                                    set AllEquippedPetsIndexes[GeneralSelector_UISlotSelected] = GeneralSelector_CurrentOption #Set the pet index on the Equipped array
                                    NewPlayerData := GameManager.PlayersDataManager.SavePets(ValidBakery.PlayerDataCache,AllPets) #Set the new all pets data
                                    #GameManager.PlayersDataManager.SaveNewData(Player,PlayerData,NewPlayerData) #Save the new Data
                                then:
                                    ValidBakery.SetNewPlayerDataCache(NewPlayerData)
                                    LoadEquippedPets(NewPlayerData)
                                    EquipPet_Display_Canvas()
                            else:
                                PrintFatalError("Pet with index {PetSlotIndex} not found in player data table")
                        else:
                            #Print("Empty slot InUISlotSelect: {GeneralSelector_UISlotSelected}: is going to be assigned with pet `{GeneralSelector_CurrentOption}`")
                            if(NewPetToUseNow:=AllPets[GeneralSelector_CurrentOption]):
                                UpdatedNewPet:=GameManager.PlayersDataManager.SetEquippedPet(NewPetToUseNow, true) #Lets put this pet to false
                                
                                
                                if(set AllPets[GeneralSelector_CurrentOption] = UpdatedNewPet):
                                    #Print("AllEquippedPetsIndexes.Length =  {AllEquippedPetsIndexes.Length}")
                                    if(set AllEquippedPetsIndexes[GeneralSelector_UISlotSelected] = GeneralSelector_CurrentOption):
                                        NewPlayerData := GameManager.PlayersDataManager.SavePets(ValidBakery.PlayerDataCache,AllPets) #Set the new all pets data
                                        ValidBakery.SetNewPlayerDataCache(NewPlayerData)
                                        #GameManager.PlayersDataManager.SaveNewData(Player,PlayerData,NewPlayerData) #Save the new Data
                                        EquipPet_Display_Canvas()        
                                    else:
                                        PrintFatalError("Assignation failed!. GeneralSelector_UISlotSelected: {GeneralSelector_UISlotSelected}. GeneralSelector_CurrentOption: {GeneralSelector_CurrentOption}")
                                else:
                                    PrintFatalError("Assignation failed!. GeneralSelector_CurrentOption: {GeneralSelector_CurrentOption}")

                                <#
                                if:
                                    set AllPets[GeneralSelector_CurrentOption] = UpdatedNewPet #And lets save if after that
                                    set AllEquippedPetsIndexes[GeneralSelector_UISlotSelected] = GeneralSelector_CurrentOption #Set the pet index on the Equipped array
                                    NewPlayerData := TestManager.PlayersDataManager.SavePets(PlayerData,AllPets) #Set the new all pets data
                                    TestManager.PlayersDataManager.SaveNewData(Player,PlayerData,NewPlayerData) #Save the new Data
                                then:
                                    EquipPet_Display_Canvas()
                                else:
                                    PrintFatalError("GeneralSelector_CurrentOption: {GeneralSelector_CurrentOption} or, GeneralSelector_UISlotSelected: {GeneralSelector_UISlotSelected} not found")
                                    #>
                            else:
                                PrintFatalError("Pet with index {GeneralSelector_CurrentOption} not found in player data table")
                            #Now, lest set the selected pet as being use
                            #Let substitute thisa pet for another
            #########
            #SUBMENU MIX PETS
            if(GeneralSelector_FromSubmenu=SUBMENU_MIX_PET):
                if(PetSlotIndex:=AllMixPets_PetsToMix[GeneralSelector_UISlotSelected], PetSlotIndex<>-1):
                    #Print("Pet of MixSlot: {GeneralSelector_UISlotSelected}: `{PetSlotIndex}` is going to be replaced by `{GeneralSelector_CurrentOption}`")
                    if(ValidBakery := MaybeBakery?, PetTable:=ValidBakery.PlayerDataCache.AllPets[PetSlotIndex]):
                        if:
                            set AllMixPets_PetsToMix[GeneralSelector_UISlotSelected] = GeneralSelector_CurrentOption #Set the pet index on the Equipped array
                else:
                    #Print("Empty slot MixSlot: {GeneralSelector_UISlotSelected}: is going to be assigned with pet `{GeneralSelector_CurrentOption}`")
                        if:
                            set AllMixPets_PetsToMix[GeneralSelector_UISlotSelected] = GeneralSelector_CurrentOption #Set the pet index on the Equipped array
                    else:
                        PrintFatalError("Pet with index {GeneralSelector_CurrentOption} not found in player data table")
                    #Now, lest set the selected pet as being use
                    #Let substitute thisa pet for another
                MixPets_Display_Canvas()
        Generate_PlayerEnhacements()
               

    GeneralSelector_NavidateRight(Message : widget_message) : void =
        if(Message.Player.IsActive[], ValidBakery := GameManager.GetAgentBakery_Transacted(agent[Message.Player])?):
            set GeneralSelector_CurrentOption = GeneralSelector_CurrentOption + 1       
            if(GeneralSelector_CurrentOption > ValidBakery.PlayerDataCache.AllPets.Length-1):
                set GeneralSelector_CurrentOption = ValidBakery.PlayerDataCache.AllPets.Length-1
        GeneralSelector_RefreshAll()

    GeneralSelector_NavidateLeft(Message : widget_message) : void =
        set GeneralSelector_CurrentOption = GeneralSelector_CurrentOption - 1
        if(GeneralSelector_CurrentOption < 0):
            set GeneralSelector_CurrentOption = 0
        GeneralSelector_RefreshAll()

    GeneralSelector_ExitPetSelector(Message : widget_message) : void =
        GeneralSelector_Remove_Canvas(Message)

    GeneralSelector_Remove_Canvas(Message : widget_message) : void =
        set GeneralSelector_UIIsDisplayed = false
        if:
            Agent := MaybeAgent?
            Player:=player[Agent]
            PlayerUI:=GetPlayerUI[Player]
            Canvas := GeneralSelector_MaybeCanvas?
        then:
            PlayerUI.RemoveWidget(Canvas)
            GeneralSelector_RemoveDelay_OnRemove()

    GeneralSelector_Remove_CanvasWithoutMessage() : void =
        set GeneralSelector_UIIsDisplayed = false
        if:
            Agent := MaybeAgent?
            Player:=player[Agent]
            PlayerUI:=GetPlayerUI[Player]
            Canvas := GeneralSelector_MaybeCanvas?
        then:
            PlayerUI.RemoveWidget(Canvas)
            GeneralSelector_RemoveDelay_OnRemove()

    GeneralSelector_RemoveDelay_OnAdd() : void =
        GeneralSelector_CancelButton.SetText(BasicMessage("Close"))

    GeneralSelector_RemoveDelay_OnRemove() : void =
        GeneralSelector_CancelButton.SetText(BasicMessage("Close "))
  
    GeneralSelector_Display_Canvas(InFromSubmenu : int, InUISlotSelect : int)<suspends> : void =
            var ErrorOnDisplay : logic = false
            set GeneralSelector_UISlotSelected = InUISlotSelect
            set GeneralSelector_FromSubmenu = InFromSubmenu

            #Print("Trying to display generla selector")
            if(GeneralSelector_FromSubmenu=SUBMENU_EQUIP_PET):
                if(Agent := MaybeAgent?, Player:=player[Agent], Player.IsActive[], ValidBakery := GameManager.GetAgentBakery_Transacted(Agent)?, InUISlotSelect<=ValidBakery.PlayerDataCache.MaxActive_PetSlots-1):
                    EquipPet_Remove_CanvasWithoutMessage()
                else:
                    set ErrorOnDisplay = true
                    PrintFatalError("Slot {InUISlotSelect} not available on GeneralSelector_Display_Canvas from Submenu {InFromSubmenu}")
            else if(GeneralSelector_FromSubmenu=SUBMENU_MIX_PET):
                MixPets_Remove_CanvasWithoutMessage()

        if(not ErrorOnDisplay?, not GeneralSelector_UIIsDisplayed?):
            if(not GeneralSelector_UIIsDisplayed?):
                set GeneralSelector_UIIsDisplayed = true
            Sleep(0.04)
            var Index : int = 0 
            if:
                Agent := MaybeAgent?
                PlayerUI:=GetPlayerUI[player[Agent]]
                Canvas := GeneralSelector_MaybeCanvas?
            then:
                PlayerUI.AddWidget(Canvas,player_ui_slot{ZOrder:=0, InputMode := ui_input_mode.All})
                GeneralSelector_RemoveDelay_OnAdd()
        else:
            PrintWarning("Slot {InUISlotSelect} not available on GeneralSelector_Display_Canvas from Submenu {InFromSubmenu}")
        GeneralSelector_RefreshAll()

    MakeGeneralSelectorCanvas(): canvas=
        Canvas := canvas:
            Slots:=array:
                #Main overlay
                canvas_slot:
                    Anchors:=anchors{Minimum:=vector2{X:=0.5,Y:=0.5},Maximum:=vector2{X:=0.5,Y:=0.5}}
                    ZOrder:=0
                    SizeToContent:=true
                    Offsets:=margin{Left:=0.0,Top:=0.0,Right:=0.0,Bottom:=0.0} #PositionX, PositionY, SizeX, SizeY
                    Alignment:=vector2{X:=0.5,Y:=0.5}
                    Widget:=texture_block{DefaultDesiredSize:=vector2{X:=3000.0,Y:=1250.0},DefaultImage:=Images.Img_MainBackground}                    
                canvas_slot:
                    Anchors:=anchors{Minimum:=vector2{X:=0.5,Y:=0.5},Maximum:=vector2{X:=0.5,Y:=0.5}}
                    ZOrder:=0
                    SizeToContent:=true
                    Offsets:=margin{Left:=0.0,Top:=-75.0,Right:=0.0,Bottom:=0.0} #PositionX, PositionY, SizeX, SizeY
                    Alignment:=vector2{X:=0.5,Y:=0.5}
                    Widget:=overlay:
                        Slots:=array:
                            overlay_slot:
                                HorizontalAlignment:=horizontal_alignment.Center
                                VerticalAlignment:=vertical_alignment.Center
                                Widget:=stack_box:  
                                    Orientation:=orientation.Horizontal
                                        Slots:=array:
                                            stack_box_slot: #Slot 0
                                                Padding:=margin{Right:=SuperLowPadding,Left:=SuperLowPadding}
                                                HorizontalAlignment:=horizontal_alignment.Center
                                                VerticalAlignment:=vertical_alignment.Bottom
                                                Widget:=stack_box:  
                                                    Orientation:=orientation.Vertical
                                                        Slots:=array:
                                                            stack_box_slot: 
                                                                Padding:=margin{Top:=SelectorBoxPadding,Bottom:=SelectorBoxPadding}
                                                                Widget:=overlay:
                                                                    Slots:=array:
                                                                        overlay_slot:
                                                                            HorizontalAlignment:=horizontal_alignment.Fill
                                                                            VerticalAlignment:=vertical_alignment.Fill
                                                                            Widget:=texture_block{DefaultImage:=Images.Img_LineBackground}
                                                                        overlay_slot:
                                                                            HorizontalAlignment:=horizontal_alignment.Fill
                                                                            VerticalAlignment:=vertical_alignment.Fill
                                                                            Padding:=margin{Left:=SelectorBoxPadding,Right:=SelectorBoxPadding,Top:=SelectorBoxPadding,Bottom:=SelectorBoxPadding}
                                                                            Widget:=texture_block{DefaultImage:=Images.Img_LineBackground}
                                                                        overlay_slot:
                                                                            HorizontalAlignment:=horizontal_alignment.Center
                                                                            VerticalAlignment:=vertical_alignment.Center
                                                                            Widget:=texture_block{DefaultImage:=Images.EmptyTexture,DefaultDesiredSize:=vector2{X:=258.0,Y:=300.0}}
                                                                        overlay_slot:
                                                                            HorizontalAlignment:=horizontal_alignment.Center
                                                                            VerticalAlignment:=vertical_alignment.Center
                                                                            Widget:=stack_box:
                                                                                Orientation:=orientation.Vertical
                                                                                Slots:=array:
                                                                                    stack_box_slot: 
                                                                                        Padding:=RowPadding
                                                                                        HorizontalAlignment:=horizontal_alignment.Center
                                                                                        VerticalAlignment:=vertical_alignment.Center
                                                                                        Widget:=GeneralSelector_TextureBlock_PetImages_Box0
                                                                                    stack_box_slot: 
                                                                                        HorizontalAlignment:=horizontal_alignment.Center
                                                                                        VerticalAlignment:=vertical_alignment.Center
                                                                                        #Padding:=margin{Left:=TextPetsBoxPadding}
                                                                                        Widget:=stack_box:
                                                                                            Orientation:=orientation.Horizontal
                                                                                            Slots:=array:
                                                                                                stack_box_slot: 
                                                                                                    Padding:=margin{Left:=2.5,Right:=2.5}
                                                                                                    Widget:=GeneralSelector_TextBlock_Classes_Box0
                                                                                                stack_box_slot: 
                                                                                                    Padding:=margin{Left:=2.5, Right:=1.0}
                                                                                                    Widget:=GeneralSelector_TextBlock_PackA_Box0
                                                                                                stack_box_slot: 
                                                                                                    Padding:=margin{Right:=2.5}
                                                                                                    Widget:=GeneralSelector_TextBlock_PackB_Box0
                                                                                    stack_box_slot: 
                                                                                        Padding:=RowPadding
                                                                                        HorizontalAlignment:=horizontal_alignment.Center
                                                                                        VerticalAlignment:=vertical_alignment.Center
                                                                                        Widget:=GeneralSelector_TextBlock_Purity_Box0
                                                                                    stack_box_slot: 
                                                                                        Padding:=RowPadding
                                                                                        HorizontalAlignment:=horizontal_alignment.Center
                                                                                        VerticalAlignment:=vertical_alignment.Center
                                                                                        Widget:=GeneralSelector_TextBlock_Tier_Box0                                    
                                                            stack_box_slot: 
                                                                Widget:=overlay:
                                                                    Slots:=array:
                                                                        overlay_slot:
                                                                            HorizontalAlignment:=horizontal_alignment.Fill
                                                                            VerticalAlignment:=vertical_alignment.Fill
                                                                            Widget:=texture_block{DefaultImage:=Images.EmptyTexture,DefaultDesiredSize:=QuietButton_PetSelectorDesiredSize}
                                            stack_box_slot: #Slot 1
                                                Padding:=margin{Right:=SuperLowPadding,Left:=SuperLowPadding}
                                                HorizontalAlignment:=horizontal_alignment.Center
                                                VerticalAlignment:=vertical_alignment.Bottom
                                                Widget:=stack_box:  
                                                    Orientation:=orientation.Vertical
                                                        Slots:=array:
                                                            stack_box_slot: 
                                                                Padding:=margin{Top:=SelectorBoxPadding,Bottom:=SelectorBoxPadding}
                                                                Widget:=overlay:
                                                                    Slots:=array:
                                                                        overlay_slot:
                                                                            HorizontalAlignment:=horizontal_alignment.Fill
                                                                            VerticalAlignment:=vertical_alignment.Fill
                                                                            Widget:=texture_block{DefaultImage:=Images.Img_LineBackground}
                                                                        overlay_slot:
                                                                            HorizontalAlignment:=horizontal_alignment.Fill
                                                                            VerticalAlignment:=vertical_alignment.Fill
                                                                            Padding:=margin{Left:=SelectorBoxPadding,Right:=SelectorBoxPadding,Top:=SelectorBoxPadding,Bottom:=SelectorBoxPadding}
                                                                            Widget:=texture_block{DefaultImage:=Images.Img_LineBackground}
                                                                        overlay_slot:
                                                                            HorizontalAlignment:=horizontal_alignment.Center
                                                                            VerticalAlignment:=vertical_alignment.Center
                                                                            Widget:=texture_block{DefaultImage:=Images.EmptyTexture,DefaultDesiredSize:=vector2{X:=258.0,Y:=300.0}}
                                                                        overlay_slot:
                                                                            HorizontalAlignment:=horizontal_alignment.Center
                                                                            VerticalAlignment:=vertical_alignment.Center
                                                                            Widget:=stack_box:
                                                                                Orientation:=orientation.Vertical
                                                                                Slots:=array:
                                                                                    stack_box_slot: 
                                                                                        Padding:=RowPadding
                                                                                        HorizontalAlignment:=horizontal_alignment.Center
                                                                                        VerticalAlignment:=vertical_alignment.Center
                                                                                        Widget:=GeneralSelector_TextureBlock_PetImages_Box1
                                                                                    stack_box_slot: 
                                                                                        HorizontalAlignment:=horizontal_alignment.Left
                                                                                        VerticalAlignment:=vertical_alignment.Center
                                                                                        Padding:=margin{Left:=TextPetsBoxPadding}
                                                                                        Widget:=stack_box:
                                                                                            Orientation:=orientation.Horizontal
                                                                                            Slots:=array:
                                                                                                stack_box_slot: 
                                                                                                    HorizontalAlignment:=horizontal_alignment.Center
                                                                                                    VerticalAlignment:=vertical_alignment.Center
                                                                                                    #Padding:=margin{Left:=TextPetsBoxPadding}
                                                                                                    Widget:=stack_box:
                                                                                                        Orientation:=orientation.Horizontal
                                                                                                        Slots:=array:
                                                                                                            stack_box_slot: 
                                                                                                                Padding:=margin{Left:=2.5,Right:=2.5}
                                                                                                                Widget:=GeneralSelector_TextBlock_Classes_Box1
                                                                                                            stack_box_slot: 
                                                                                                                Padding:=margin{Left:=2.5, Right:=1.0}
                                                                                                                Widget:=GeneralSelector_TextBlock_PackA_Box1
                                                                                                            stack_box_slot: 
                                                                                                                Padding:=margin{Right:=2.5}
                                                                                                                Widget:=GeneralSelector_TextBlock_PackB_Box1
                                                                                                stack_box_slot: 
                                                                                                    Padding:=margin{Left:=2.5, Right:=1.0}
                                                                                                    Widget:=GeneralSelector_TextBlock_PackA_Box1
                                                                                                stack_box_slot: 
                                                                                                    Padding:=margin{Right:=2.5}
                                                                                                    Widget:=GeneralSelector_TextBlock_PackB_Box1
                                                                                    stack_box_slot: 
                                                                                        Padding:=RowPadding
                                                                                        HorizontalAlignment:=horizontal_alignment.Center
                                                                                        VerticalAlignment:=vertical_alignment.Center
                                                                                        Widget:=GeneralSelector_TextBlock_Purity_Box1
                                                                                    stack_box_slot: 
                                                                                        Padding:=RowPadding
                                                                                        HorizontalAlignment:=horizontal_alignment.Center
                                                                                        VerticalAlignment:=vertical_alignment.Center
                                                                                        Widget:=GeneralSelector_TextBlock_Tier_Box1
                                                            stack_box_slot: 
                                                                Widget:=overlay:
                                                                    Slots:=array:
                                                                        overlay_slot:
                                                                            HorizontalAlignment:=horizontal_alignment.Fill
                                                                            VerticalAlignment:=vertical_alignment.Fill
                                                                            Widget:=texture_block{DefaultImage:=Images.EmptyTexture,DefaultDesiredSize:=QuietButton_PetSelectorDesiredSize}

                                            stack_box_slot: #SELECTED
                                                Padding:=margin{Right:=SuperLowPadding,Left:=SuperLowPadding}
                                                HorizontalAlignment:=horizontal_alignment.Center
                                                VerticalAlignment:=vertical_alignment.Bottom
                                                Widget:=stack_box:  
                                                    Orientation:=orientation.Vertical
                                                    Slots:=array:
                                                        stack_box_slot: 
                                                            Padding:=margin{Top:=SelectorBoxPadding,Bottom:=SelectorBoxPadding}
                                                            Widget:=overlay:
                                                                Slots:=array:
                                                                    overlay_slot:
                                                                        HorizontalAlignment:=horizontal_alignment.Fill
                                                                        VerticalAlignment:=vertical_alignment.Fill
                                                                        Widget:=texture_block{DefaultImage:=Images.Img_LineBackground}
                                                                    overlay_slot:
                                                                        HorizontalAlignment:=horizontal_alignment.Fill
                                                                        VerticalAlignment:=vertical_alignment.Fill
                                                                        Padding:=margin{Left:=SelectorBoxPadding,Right:=SelectorBoxPadding,Top:=SelectorBoxPadding,Bottom:=SelectorBoxPadding}
                                                                        Widget:=texture_block{DefaultImage:=Images.Img_LineBackground}
                                                                    overlay_slot:
                                                                        HorizontalAlignment:=horizontal_alignment.Center
                                                                        VerticalAlignment:=vertical_alignment.Center
                                                                        Widget:=texture_block{DefaultImage:=Images.EmptyTexture,DefaultDesiredSize:=vector2{X:=1100.0,Y:=705.0}}
                                                                    overlay_slot:
                                                                        HorizontalAlignment:=horizontal_alignment.Center
                                                                        VerticalAlignment:=vertical_alignment.Center
                                                                        Widget:=stack_box:  
                                                                            Orientation:=orientation.Vertical
                                                                            Slots:=array:
                                                                                stack_box_slot: 
                                                                                    Widget:=stack_box:  
                                                                                        Orientation:=orientation.Horizontal
                                                                                        Slots:=array:
                                                                                            stack_box_slot: 
                                                                                                Padding:=RowPadding
                                                                                                Widget:=GeneralSelector_Selected_TextureBlock_PetImages
                                                                                            stack_box_slot: 
                                                                                                HorizontalAlignment:=horizontal_alignment.Center
                                                                                                VerticalAlignment:=vertical_alignment.Center
                                                                                                Widget:=stack_box:  
                                                                                                    Orientation:=orientation.Vertical
                                                                                                    Slots:=array:
                                                                                                        stack_box_slot: 
                                                                                                            HorizontalAlignment:=horizontal_alignment.Left
                                                                                                            VerticalAlignment:=vertical_alignment.Center
                                                                                                            Padding:=margin{Left:=TextPetsBoxPadding}
                                                                                                            Widget:=stack_box:
                                                                                                                Orientation:=orientation.Horizontal
                                                                                                                Slots:=array:
                                                                                                                    stack_box_slot: 
                                                                                                                        HorizontalAlignment:=horizontal_alignment.Center
                                                                                                                        VerticalAlignment:=vertical_alignment.Center
                                                                                                                        #Padding:=margin{Left:=TextPetsBoxPadding}
                                                                                                                        Widget:=stack_box:
                                                                                                                            Orientation:=orientation.Horizontal
                                                                                                                            Slots:=array:
                                                                                                                                stack_box_slot: 
                                                                                                                                    Padding:=margin{Left:=2.5,Right:=2.5}
                                                                                                                                    Widget:=GeneralSelector_Selected_TextBlock_Class
                                                                                                                                stack_box_slot: 
                                                                                                                                    Padding:=margin{Left:=2.5, Right:=1.0}
                                                                                                                                    Widget:=GeneralSelector_Selected_TextBlock_PackA
                                                                                                                                stack_box_slot: 
                                                                                                                                    Padding:=margin{Right:=2.5}
                                                                                                                                    Widget:=GeneralSelector_Selected_TextBlock_PackB
                                                                                                                    stack_box_slot: 
                                                                                                                        Padding:=margin{Left:=2.5, Right:=1.0}
                                                                                                                        Widget:=GeneralSelector_Selected_TextBlock_PackA
                                                                                                                    stack_box_slot: 
                                                                                                                        Padding:=margin{Right:=2.5}
                                                                                                                        Widget:=GeneralSelector_Selected_TextBlock_PackB
                                                                                                        stack_box_slot: 
                                                                                                            Padding:=RowPadding
                                                                                                            Widget:=GeneralSelector_Selected_TextBlock_Purity
                                                                                stack_box_slot: 
                                                                                    HorizontalAlignment:=horizontal_alignment.Fill
                                                                                    VerticalAlignment:=vertical_alignment.Fill
                                                                                    Widget:=texture_block{DefaultImage:=Images.Img_HorizontalLine,DefaultDesiredSize:=vector2{Y:=3.0}}
                                                                                stack_box_slot:  
                                                                                    HorizontalAlignment:=horizontal_alignment.Center
                                                                                    VerticalAlignment:=vertical_alignment.Center
                                                                                    Padding:=RowPadding
                                                                                    Widget:=GeneralSelector_Text_PackageA_LV
                                                                                stack_box_slot: 
                                                                                    HorizontalAlignment:=horizontal_alignment.Fill
                                                                                    VerticalAlignment:=vertical_alignment.Fill
                                                                                    Widget:=texture_block{DefaultImage:=Images.Img_HorizontalLine,DefaultDesiredSize:=vector2{Y:=3.0}}
                                                                                stack_box_slot: 
                                                                                    HorizontalAlignment:=horizontal_alignment.Fill
                                                                                    VerticalAlignment:=vertical_alignment.Fill
                                                                                    Widget:=stack_box:  
                                                                                        Orientation:=orientation.Horizontal
                                                                                        Slots:=array:
                                                                                            stack_box_slot: 
                                                                                                Widget:=stack_box:
                                                                                                    Orientation:=orientation.Vertical
                                                                                                    Slots:=array:
                                                                                                        stack_box_slot: 
                                                                                                            Padding:=RowPadding
                                                                                                            Widget:=text_block{DefaultText:=BasicMessage("Tournament stats"), DefaultTextColor:=Red, DefaultShadowOffset:=ShadowOffset, DefaultShadowOpacity := DefaultTextShadowOpacity}
                                                                                                        stack_box_slot: 
                                                                                                            HorizontalAlignment:=horizontal_alignment.Fill
                                                                                                            VerticalAlignment:=vertical_alignment.Fill
                                                                                                            Widget:=stack_box:  
                                                                                                                Orientation:=orientation.Horizontal
                                                                                                                Slots:=array:
                                                                                                                    stack_box_slot: 
                                                                                                                        HorizontalAlignment:=horizontal_alignment.Fill
                                                                                                                        VerticalAlignment:=vertical_alignment.Fill
                                                                                                                        Widget:=stack_box:  
                                                                                                                            Orientation:=orientation.Vertical
                                                                                                                            Slots:=array:
                                                                                                                                stack_box_slot: 
                                                                                                                                    Widget:=overlay:
                                                                                                                                        Slots:=array:
                                                                                                                                            overlay_slot:
                                                                                                                                                HorizontalAlignment:=horizontal_alignment.Left
                                                                                                                                                Padding:=RowPadding
                                                                                                                                                Widget:=text_block{DefaultText:=BasicMessage("Health:"), DefaultTextColor:=White, DefaultShadowOffset:=ShadowOffset, DefaultShadowOpacity := DefaultTextShadowOpacity}
                                                                                                                                            overlay_slot:
                                                                                                                                                Widget:=texture_block{DefaultImage:=Images.EmptyTexture,DefaultDesiredSize:=vector2{X:=230.0,Y:=32.0}} 
                                                                                                                                stack_box_slot: 
                                                                                                                                    HorizontalAlignment:=horizontal_alignment.Left
                                                                                                                                    Padding:=RowPadding
                                                                                                                                    Widget:=text_block{DefaultText:=BasicMessage("Damage:"), DefaultTextColor:=White, DefaultShadowOffset:=ShadowOffset, DefaultShadowOpacity := DefaultTextShadowOpacity}
                                                                                                                                stack_box_slot: 
                                                                                                                                    HorizontalAlignment:=horizontal_alignment.Left
                                                                                                                                    Padding:=RowPadding
                                                                                                                                    Widget:=text_block{DefaultText:=BasicMessage("Mov. speed:"), DefaultTextColor:=White, DefaultShadowOffset:=ShadowOffset, DefaultShadowOpacity := DefaultTextShadowOpacity}
                                                                                                                    stack_box_slot: 
                                                                                                                        HorizontalAlignment:=horizontal_alignment.Fill
                                                                                                                        VerticalAlignment:=vertical_alignment.Fill
                                                                                                                        Widget:=stack_box:  
                                                                                                                            Orientation:=orientation.Vertical
                                                                                                                            Slots:=array:
                                                                                                                                stack_box_slot: 
                                                                                                                                    Widget:=overlay:
                                                                                                                                        Slots:=array:
                                                                                                                                            overlay_slot:
                                                                                                                                                Padding:=RowPadding
                                                                                                                                                Widget:=GeneralSelector_Selected_TextBlock_Current_TournamentHealth
                                                                                                                                            overlay_slot:
                                                                                                                                                Widget:=texture_block{DefaultImage:=Images.EmptyTexture,DefaultDesiredSize:=vector2{X:=257.0,Y:=32.0}}
                                                                                                                                stack_box_slot: 
                                                                                                                                    Padding:=RowPadding
                                                                                                                                    Widget:=GeneralSelector_Selected_TextBlock_Current_TournamentDamage
                                                                                                                                stack_box_slot: 
                                                                                                                                    Padding:=RowPadding
                                                                                                                                    Widget:=GeneralSelector_Selected_TextBlock_Current_TournamentSpeed
                                                                                            stack_box_slot: 
                                                                                                HorizontalAlignment:=horizontal_alignment.Center
                                                                                                VerticalAlignment:=vertical_alignment.Fill
                                                                                                Widget:=texture_block{DefaultImage:=Images.Img_VerticalLine,DefaultDesiredSize:=vector2{X:=10.0}}
                                                                                            stack_box_slot: 
                                                                                                Widget:=stack_box:
                                                                                                    Orientation:=orientation.Vertical
                                                                                                    Slots:=array:
                                                                                                        stack_box_slot: 
                                                                                                            Padding:=RowPadding
                                                                                                            Widget:=text_block{DefaultText:=BasicMessage("Player enhancements"), DefaultTextColor:=Cyan, DefaultShadowOffset:=ShadowOffset, DefaultShadowOpacity := DefaultTextShadowOpacity}
                                                                                                        stack_box_slot: 
                                                                                                            HorizontalAlignment:=horizontal_alignment.Fill
                                                                                                            VerticalAlignment:=vertical_alignment.Fill
                                                                                                            Widget:=stack_box:  
                                                                                                                Orientation:=orientation.Horizontal
                                                                                                                Slots:=array:
                                                                                                                    stack_box_slot: 
                                                                                                                        HorizontalAlignment:=horizontal_alignment.Fill
                                                                                                                        VerticalAlignment:=vertical_alignment.Fill
                                                                                                                        Widget:=stack_box:  
                                                                                                                            Orientation:=orientation.Vertical
                                                                                                                            Slots:=array:
                                                                                                                                stack_box_slot: 
                                                                                                                                    Widget:=overlay:
                                                                                                                                        Slots:=array:
                                                                                                                                            overlay_slot:
                                                                                                                                                Padding:=RowPadding
                                                                                                                                                HorizontalAlignment:=horizontal_alignment.Left
                                                                                                                                                Widget:=GeneralSelector_Selected_TextBlock_Description_SpecialPerkA
                                                                                                                                            overlay_slot:
                                                                                                                                                Widget:=texture_block{DefaultImage:=Images.EmptyTexture,DefaultDesiredSize:=vector2{X:=300.0,Y:=32.0}}
                                                                                                                                stack_box_slot: 
                                                                                                                                    Padding:=RowPadding
                                                                                                                                    HorizontalAlignment:=horizontal_alignment.Left
                                                                                                                                    Widget:=text_block{DefaultText:=BasicMessage("Pick damage:"), DefaultTextColor:=White, DefaultShadowOffset:=ShadowOffset, DefaultShadowOpacity := DefaultTextShadowOpacity}
                                                                                                                                stack_box_slot: 
                                                                                                                                    Padding:=RowPadding
                                                                                                                                    HorizontalAlignment:=horizontal_alignment.Left
                                                                                                                                    Widget:=text_block{DefaultText:=BasicMessage("Weak point mult:"), DefaultTextColor:=White, DefaultShadowOffset:=ShadowOffset, DefaultShadowOpacity := DefaultTextShadowOpacity}
                                                                                                                    stack_box_slot: 
                                                                                                                        HorizontalAlignment:=horizontal_alignment.Fill
                                                                                                                        VerticalAlignment:=vertical_alignment.Fill
                                                                                                                        Widget:=stack_box:  
                                                                                                                            Orientation:=orientation.Vertical
                                                                                                                            Slots:=array:
                                                                                                                                stack_box_slot: 
                                                                                                                                    Widget:=overlay:
                                                                                                                                        Slots:=array:
                                                                                                                                            overlay_slot:
                                                                                                                                                Padding:=RowPadding
                                                                                                                                                Widget:=GeneralSelector_Selected_TextBlock_Current_SpecialPerkA
                                                                                                                                            overlay_slot:
                                                                                                                                                Widget:=texture_block{DefaultImage:=Images.EmptyTexture,DefaultDesiredSize:=vector2{X:=180.0,Y:=32.0}}
                                                                                                                                stack_box_slot: 
                                                                                                                                    Padding:=RowPadding
                                                                                                                                    Widget:=GeneralSelector_Selected_TextBlock_Current_PlayerPickDamage
                                                                                                                                stack_box_slot:
                                                                                                                                    Padding:=RowPadding 
                                                                                                                                    Widget:=GeneralSelector_Selected_TextBlock_Current_PlayerWeakPoint
                                                                                stack_box_slot: 
                                                                                    HorizontalAlignment:=horizontal_alignment.Fill
                                                                                    VerticalAlignment:=vertical_alignment.Fill
                                                                                    Widget:=texture_block{DefaultImage:=Images.Img_HorizontalLine,DefaultDesiredSize:=vector2{Y:=3.0}}
                                                                                stack_box_slot:  
                                                                                    HorizontalAlignment:=horizontal_alignment.Center
                                                                                    VerticalAlignment:=vertical_alignment.Center
                                                                                    Padding:=RowPadding
                                                                                    Widget:=GeneralSelector_Text_PackageB_LV
                                                                                stack_box_slot: 
                                                                                    HorizontalAlignment:=horizontal_alignment.Fill
                                                                                    VerticalAlignment:=vertical_alignment.Fill
                                                                                    Widget:=texture_block{DefaultImage:=Images.Img_HorizontalLine,DefaultDesiredSize:=vector2{Y:=3.0}}
                                                                                stack_box_slot: 
                                                                                    HorizontalAlignment:=horizontal_alignment.Fill
                                                                                    VerticalAlignment:=vertical_alignment.Fill
                                                                                    Widget:=stack_box:  
                                                                                        Orientation:=orientation.Horizontal
                                                                                        Slots:=array:
                                                                                            stack_box_slot: 
                                                                                                Widget:=stack_box:
                                                                                                    Orientation:=orientation.Vertical
                                                                                                    Slots:=array:
                                                                                                        stack_box_slot: 
                                                                                                            Padding:=RowPadding
                                                                                                            Widget:=text_block{DefaultText:=BasicMessage("Tournament stats"), DefaultTextColor:=Red, DefaultShadowOffset:=ShadowOffset, DefaultShadowOpacity := DefaultTextShadowOpacity}
                                                                                                        stack_box_slot: 
                                                                                                            HorizontalAlignment:=horizontal_alignment.Fill
                                                                                                            VerticalAlignment:=vertical_alignment.Fill
                                                                                                            Widget:=stack_box:  
                                                                                                                Orientation:=orientation.Horizontal
                                                                                                                Slots:=array:
                                                                                                                    stack_box_slot: 
                                                                                                                        HorizontalAlignment:=horizontal_alignment.Fill
                                                                                                                        VerticalAlignment:=vertical_alignment.Fill
                                                                                                                        Widget:=stack_box:  
                                                                                                                            Orientation:=orientation.Vertical
                                                                                                                            Slots:=array:
                                                                                                                                stack_box_slot: 
                                                                                                                                    Widget:=overlay:
                                                                                                                                        Slots:=array:
                                                                                                                                            overlay_slot:
                                                                                                                                                Padding:=RowPadding
                                                                                                                                                HorizontalAlignment:=horizontal_alignment.Left
                                                                                                                                                Widget:=text_block{DefaultText:=BasicMessage("Attack time:"), DefaultTextColor:=White, DefaultShadowOffset:=ShadowOffset, DefaultShadowOpacity := DefaultTextShadowOpacity}
                                                                                                                                            overlay_slot:
                                                                                                                                                Widget:=texture_block{DefaultImage:=Images.EmptyTexture,DefaultDesiredSize:=vector2{X:=230.0,Y:=32.0}}
                                                                                                                                stack_box_slot: 
                                                                                                                                    Padding:=RowPadding
                                                                                                                                    HorizontalAlignment:=horizontal_alignment.Left
                                                                                                                                    Widget:=text_block{DefaultText:=BasicMessage("Range:"), DefaultTextColor:=White, DefaultShadowOffset:=ShadowOffset, DefaultShadowOpacity := DefaultTextShadowOpacity}
                                                                                                                                stack_box_slot: 
                                                                                                                                    Padding:=RowPadding
                                                                                                                                    HorizontalAlignment:=horizontal_alignment.Left
                                                                                                                                    Widget:=text_block{DefaultText:=BasicMessage("Healing perk:"), DefaultTextColor:=White, DefaultShadowOffset:=ShadowOffset, DefaultShadowOpacity := DefaultTextShadowOpacity}
                                                                                                                    stack_box_slot: 
                                                                                                                        HorizontalAlignment:=horizontal_alignment.Fill
                                                                                                                        VerticalAlignment:=vertical_alignment.Fill
                                                                                                                        Widget:=stack_box:  
                                                                                                                            Orientation:=orientation.Vertical
                                                                                                                            Slots:=array:
                                                                                                                                stack_box_slot: 
                                                                                                                                    Widget:=overlay:
                                                                                                                                        Slots:=array:
                                                                                                                                            overlay_slot:
                                                                                                                                                Padding:=RowPadding
                                                                                                                                                Widget:=GeneralSelector_Selected_TextBlock_Current_TournamentActionSpeed
                                                                                                                                            overlay_slot:
                                                                                                                                                Widget:=texture_block{DefaultImage:=Images.EmptyTexture,DefaultDesiredSize:=vector2{X:=257.0,Y:=32.0}}
                                                                                                                                stack_box_slot: 
                                                                                                                                    Padding:=RowPadding
                                                                                                                                    Widget:=GeneralSelector_Selected_TextBlock_Current_TournamentRange
                                                                                                                                stack_box_slot: 
                                                                                                                                    Padding:=RowPadding
                                                                                                                                    Widget:=GeneralSelector_Selected_TextBlock_Current_TournamentHealingPerk
                                                                                            stack_box_slot: 
                                                                                                HorizontalAlignment:=horizontal_alignment.Center
                                                                                                VerticalAlignment:=vertical_alignment.Fill
                                                                                                Widget:=texture_block{DefaultImage:=Images.Img_VerticalLine,DefaultDesiredSize:=vector2{X:=10.0}}
                                                                                            stack_box_slot: 
                                                                                                Widget:=stack_box:
                                                                                                    Orientation:=orientation.Vertical
                                                                                                    Slots:=array:
                                                                                                        stack_box_slot: 
                                                                                                            Padding:=RowPadding
                                                                                                            Widget:=text_block{DefaultText:=BasicMessage("Player enhancements"), DefaultTextColor:=Cyan, DefaultShadowOffset:=ShadowOffset, DefaultShadowOpacity := DefaultTextShadowOpacity}
                                                                                                        stack_box_slot: 
                                                                                                            HorizontalAlignment:=horizontal_alignment.Fill
                                                                                                            VerticalAlignment:=vertical_alignment.Fill
                                                                                                            Widget:=stack_box:  
                                                                                                                Orientation:=orientation.Horizontal
                                                                                                                Slots:=array:
                                                                                                                    stack_box_slot: 
                                                                                                                        HorizontalAlignment:=horizontal_alignment.Fill
                                                                                                                        VerticalAlignment:=vertical_alignment.Fill
                                                                                                                        Widget:=stack_box:  
                                                                                                                            Orientation:=orientation.Vertical
                                                                                                                            Slots:=array:
                                                                                                                                stack_box_slot: 
                                                                                                                                    Widget:=overlay:
                                                                                                                                        Slots:=array:
                                                                                                                                            overlay_slot:
                                                                                                                                                Padding:=RowPadding
                                                                                                                                                HorizontalAlignment:=horizontal_alignment.Left
                                                                                                                                                Widget:=GeneralSelector_Selected_TextBlock_Description_SpecialPerkB
                                                                                                                                            overlay_slot:
                                                                                                                                                Widget:=texture_block{DefaultImage:=Images.EmptyTexture,DefaultDesiredSize:=vector2{X:=300.0,Y:=32.0}}
                                                                                                                                stack_box_slot: 
                                                                                                                                    Padding:=RowPadding
                                                                                                                                    HorizontalAlignment:=horizontal_alignment.Left
                                                                                                                                    Widget:=text_block{DefaultText:=BasicMessage("Pick efficiency:"), DefaultTextColor:=White, DefaultShadowOffset:=ShadowOffset, DefaultShadowOpacity := DefaultTextShadowOpacity}
                                                                                                                                stack_box_slot: 
                                                                                                                                    Padding:=RowPadding
                                                                                                                                    HorizontalAlignment:=horizontal_alignment.Left
                                                                                                                                    Widget:=text_block{DefaultText:=BasicMessage("Pick critical dam:"), DefaultTextColor:=White, DefaultShadowOffset:=ShadowOffset, DefaultShadowOpacity := DefaultTextShadowOpacity}
                                                                                                                    stack_box_slot: 
                                                                                                                        HorizontalAlignment:=horizontal_alignment.Fill
                                                                                                                        VerticalAlignment:=vertical_alignment.Fill
                                                                                                                        Widget:=stack_box:  
                                                                                                                            Orientation:=orientation.Vertical
                                                                                                                            Slots:=array:
                                                                                                                                stack_box_slot: 
                                                                                                                                    Widget:=overlay:
                                                                                                                                        Slots:=array:
                                                                                                                                            overlay_slot:
                                                                                                                                                Padding:=RowPadding
                                                                                                                                                Widget:=GeneralSelector_Selected_TextBlock_Current_SpecialPerkB
                                                                                                                                            overlay_slot:
                                                                                                                                                Widget:=texture_block{DefaultImage:=Images.EmptyTexture,DefaultDesiredSize:=vector2{X:=180.0,Y:=32.0}}
                                                                                                                                stack_box_slot: 
                                                                                                                                    Padding:=RowPadding
                                                                                                                                    Widget:=GeneralSelector_Selected_TextBlock_Current_PlayerPickEfficiency
                                                                                                                                stack_box_slot: 
                                                                                                                                    Padding:=RowPadding
                                                                                                                                    Widget:=GeneralSelector_Selected_TextBlock_Current_PlayerPickCriticalDamage
                                                        stack_box_slot: #Buttons Here 
                                                            Padding:=margin{Top:=SelectorBoxPadding,Bottom:=SelectorBoxPadding}
                                                            Widget:=overlay:
                                                                Slots:=array:
                                                                    overlay_slot:
                                                                        Widget:=stack_box:  
                                                                            Orientation:=orientation.Horizontal
                                                                                Slots:=array:
                                                                                    stack_box_slot: 
                                                                                        Padding:=RowPadding 
                                                                                        Widget:=overlay:
                                                                                            Slots:=array:
                                                                                                overlay_slot:
                                                                                                    Widget:=texture_block{DefaultImage:=Images.EmptyTexture,DefaultDesiredSize:=vector2{X:=530.0,Y:=80.0}}
                                                                                                overlay_slot:
                                                                                                    HorizontalAlignment:=horizontal_alignment.Fill
                                                                                                    VerticalAlignment:=vertical_alignment.Fill
                                                                                                    Widget:=GeneralSelector_NavigateLeftButton
                                                                                    stack_box_slot: 
                                                                                        Padding:=RowPadding 
                                                                                        Widget:=overlay:
                                                                                            Slots:=array:
                                                                                                overlay_slot:
                                                                                                    Widget:=texture_block{DefaultImage:=Images.EmptyTexture,DefaultDesiredSize:=vector2{X:=530.0,Y:=80.0}}
                                                                                                overlay_slot:
                                                                                                    HorizontalAlignment:=horizontal_alignment.Fill
                                                                                                    VerticalAlignment:=vertical_alignment.Fill
                                                                                                    Widget:=GeneralSelector_NavigateRightButton

                                                                                                    
                                                        stack_box_slot: #Buttons Here 
                                                            Padding:=margin{Top:=SelectorBoxPadding,Bottom:=SelectorBoxPadding}
                                                            Widget:=overlay:
                                                                Slots:=array:
                                                                    overlay_slot:
                                                                        Widget:=texture_block{DefaultImage:=Images.EmptyTexture,DefaultDesiredSize:=vector2{X:=800.0,Y:=80.0}}
                                                                    overlay_slot:
                                                                        HorizontalAlignment:=horizontal_alignment.Fill
                                                                        VerticalAlignment:=vertical_alignment.Fill
                                                                        Widget:=GeneralSelector_SelectButton
                                            stack_box_slot: #Slot 2
                                                Padding:=margin{Right:=SuperLowPadding,Left:=SuperLowPadding}
                                                HorizontalAlignment:=horizontal_alignment.Center
                                                VerticalAlignment:=vertical_alignment.Bottom
                                                Widget:=stack_box:  
                                                    Orientation:=orientation.Vertical
                                                        Slots:=array:
                                                            stack_box_slot: 
                                                                Padding:=margin{Top:=SelectorBoxPadding,Bottom:=SelectorBoxPadding}
                                                                Widget:=overlay:
                                                                    Slots:=array:
                                                                        overlay_slot:
                                                                            HorizontalAlignment:=horizontal_alignment.Fill
                                                                            VerticalAlignment:=vertical_alignment.Fill
                                                                            Widget:=texture_block{DefaultImage:=Images.Img_LineBackground}
                                                                        overlay_slot:
                                                                            HorizontalAlignment:=horizontal_alignment.Fill
                                                                            VerticalAlignment:=vertical_alignment.Fill
                                                                            Padding:=margin{Left:=SelectorBoxPadding,Right:=SelectorBoxPadding,Top:=SelectorBoxPadding,Bottom:=SelectorBoxPadding}
                                                                            Widget:=texture_block{DefaultImage:=Images.Img_LineBackground}
                                                                        overlay_slot:
                                                                            HorizontalAlignment:=horizontal_alignment.Center
                                                                            VerticalAlignment:=vertical_alignment.Center
                                                                            Widget:=texture_block{DefaultImage:=Images.EmptyTexture,DefaultDesiredSize:=vector2{X:=258.0,Y:=300.0}}
                                                                        overlay_slot:
                                                                            HorizontalAlignment:=horizontal_alignment.Center
                                                                            VerticalAlignment:=vertical_alignment.Center
                                                                            Widget:=stack_box:
                                                                                Orientation:=orientation.Vertical
                                                                                Slots:=array:
                                                                                    stack_box_slot: 
                                                                                        Padding:=RowPadding
                                                                                        HorizontalAlignment:=horizontal_alignment.Center
                                                                                        VerticalAlignment:=vertical_alignment.Center
                                                                                        Widget:=GeneralSelector_TextureBlock_PetImages_Box2
                                                                                    stack_box_slot: 
                                                                                        HorizontalAlignment:=horizontal_alignment.Center
                                                                                        VerticalAlignment:=vertical_alignment.Center
                                                                                        #Padding:=margin{Left:=TextPetsBoxPadding}
                                                                                        Widget:=stack_box:
                                                                                            Orientation:=orientation.Horizontal
                                                                                            Slots:=array:
                                                                                                stack_box_slot: 
                                                                                                    Padding:=margin{Left:=2.5,Right:=2.5}
                                                                                                    Widget:=GeneralSelector_TextBlock_Classes_Box2
                                                                                                stack_box_slot: 
                                                                                                    Padding:=margin{Left:=2.5, Right:=1.0}
                                                                                                    Widget:=GeneralSelector_TextBlock_PackA_Box2
                                                                                                stack_box_slot: 
                                                                                                    Padding:=margin{Right:=2.5}
                                                                                                    Widget:=GeneralSelector_TextBlock_PackB_Box2
                                                                                    stack_box_slot: 
                                                                                        Padding:=RowPadding
                                                                                        HorizontalAlignment:=horizontal_alignment.Center
                                                                                        VerticalAlignment:=vertical_alignment.Center
                                                                                        Widget:=GeneralSelector_TextBlock_Purity_Box2
                                                                                    stack_box_slot: 
                                                                                        Padding:=RowPadding
                                                                                        HorizontalAlignment:=horizontal_alignment.Center
                                                                                        VerticalAlignment:=vertical_alignment.Center
                                                                                        Widget:=GeneralSelector_TextBlock_Tier_Box2
                                                            stack_box_slot: 
                                                                Widget:=overlay:
                                                                    Slots:=array:
                                                                        overlay_slot:
                                                                            HorizontalAlignment:=horizontal_alignment.Fill
                                                                            VerticalAlignment:=vertical_alignment.Fill
                                                                            Widget:=texture_block{DefaultImage:=Images.EmptyTexture,DefaultDesiredSize:=QuietButton_PetSelectorDesiredSize}

                                            stack_box_slot: #Slot 4
                                                Padding:=margin{Right:=SuperLowPadding,Left:=SuperLowPadding}
                                                HorizontalAlignment:=horizontal_alignment.Center
                                                VerticalAlignment:=vertical_alignment.Bottom
                                                Widget:=stack_box:  
                                                    Orientation:=orientation.Vertical
                                                        Slots:=array:
                                                            stack_box_slot: 
                                                                Padding:=margin{Top:=SelectorBoxPadding,Bottom:=SelectorBoxPadding}
                                                                Widget:=overlay:
                                                                    Slots:=array:
                                                                        overlay_slot:
                                                                            HorizontalAlignment:=horizontal_alignment.Fill
                                                                            VerticalAlignment:=vertical_alignment.Fill
                                                                            Widget:=texture_block{DefaultImage:=Images.Img_LineBackground}
                                                                        overlay_slot:
                                                                            HorizontalAlignment:=horizontal_alignment.Fill
                                                                            VerticalAlignment:=vertical_alignment.Fill
                                                                            Padding:=margin{Left:=SelectorBoxPadding,Right:=SelectorBoxPadding,Top:=SelectorBoxPadding,Bottom:=SelectorBoxPadding}
                                                                            Widget:=texture_block{DefaultImage:=Images.Img_LineBackground}
                                                                        overlay_slot:
                                                                            HorizontalAlignment:=horizontal_alignment.Center
                                                                            VerticalAlignment:=vertical_alignment.Center
                                                                            Widget:=texture_block{DefaultImage:=Images.EmptyTexture,DefaultDesiredSize:=vector2{X:=258.0,Y:=300.0}}
                                                                        overlay_slot:
                                                                            HorizontalAlignment:=horizontal_alignment.Center
                                                                            VerticalAlignment:=vertical_alignment.Center
                                                                            Widget:=stack_box:
                                                                                Orientation:=orientation.Vertical
                                                                                Slots:=array:
                                                                                    stack_box_slot: 
                                                                                        Padding:=RowPadding
                                                                                        HorizontalAlignment:=horizontal_alignment.Center
                                                                                        VerticalAlignment:=vertical_alignment.Center
                                                                                        Widget:=GeneralSelector_TextureBlock_PetImages_Box3
                                                                                    stack_box_slot: 
                                                                                        HorizontalAlignment:=horizontal_alignment.Center
                                                                                        VerticalAlignment:=vertical_alignment.Center
                                                                                        #Padding:=margin{Left:=TextPetsBoxPadding}
                                                                                        Widget:=stack_box:
                                                                                            Orientation:=orientation.Horizontal
                                                                                            Slots:=array:
                                                                                                stack_box_slot: 
                                                                                                    Padding:=margin{Left:=2.5,Right:=2.5}
                                                                                                    Widget:=GeneralSelector_TextBlock_Classes_Box3
                                                                                                stack_box_slot: 
                                                                                                    Padding:=margin{Left:=2.5, Right:=1.0}
                                                                                                    Widget:=GeneralSelector_TextBlock_PackA_Box3
                                                                                                stack_box_slot: 
                                                                                                    Padding:=margin{Right:=2.5}
                                                                                                    Widget:=GeneralSelector_TextBlock_PackB_Box3
                                                                                    stack_box_slot: 
                                                                                        Padding:=RowPadding
                                                                                        HorizontalAlignment:=horizontal_alignment.Center
                                                                                        VerticalAlignment:=vertical_alignment.Center
                                                                                        Widget:=GeneralSelector_TextBlock_Purity_Box3
                                                                                    stack_box_slot: 
                                                                                        Padding:=RowPadding
                                                                                        HorizontalAlignment:=horizontal_alignment.Center
                                                                                        VerticalAlignment:=vertical_alignment.Center
                                                                                        Widget:=GeneralSelector_TextBlock_Tier_Box3
                                                            stack_box_slot: 
                                                                Widget:=overlay:
                                                                    Slots:=array:
                                                                        overlay_slot:
                                                                            HorizontalAlignment:=horizontal_alignment.Fill
                                                                            VerticalAlignment:=vertical_alignment.Fill
                                                                            Widget:=texture_block{DefaultImage:=Images.EmptyTexture,DefaultDesiredSize:=QuietButton_PetSelectorDesiredSize}
                canvas_slot:
                    Anchors:=anchors{Minimum:=vector2{X:=0.5,Y:=0.5},Maximum:=vector2{X:=0.5,Y:=0.5}}
                    ZOrder:=0
                    SizeToContent:=true
                    Offsets:=margin{Left:=0.0,Top:=420.0,Right:=0.0,Bottom:=0.0} #PositionX, PositionY, SizeX, SizeY
                    Alignment:=vector2{X:=0.5,Y:=0.5}
                    Widget:=overlay:
                        Slots:=array:
                            overlay_slot:
                                Padding:=margin{Top:=RowTopHugePadding,Bottom:=RowTopHugePadding}
                                Widget:=stack_box:  
                                    Orientation:=orientation.Vertical
                                        Slots:=array:                                    
                                            stack_box_slot: 
                                                    Padding:=margin{Top:=RowTopHugePadding,Bottom:=RowTopHugePadding}
                                                    Widget:=overlay:
                                                        Slots:=array:
                                                            overlay_slot:
                                                                Widget:=texture_block{DefaultImage:=Images.EmptyTexture,DefaultDesiredSize:=vector2{X:=80.0}} 
                                                            overlay_slot:
                                                                HorizontalAlignment:=horizontal_alignment.Fill
                                                                VerticalAlignment:=vertical_alignment.Fill
                                                                Widget:=GeneralSelector_CancelButton
    CreatePet_RefreshAll() : void =
        if:
            NumTiers:=Floor[(CreatePet_PetToCreate_CurrentPurity*1.0)/100.0]+1
            NumPurity:=Floor[(CreatePet_PetToCreate_CurrentPurity*1.0)-(100.0*(NumTiers-1))]
        then:
            set CreatePet_PetToCreate_CurrentTier = NumTiers
            set CreatePet_TotalCost = CalculePetCost(CreatePet_PetToCreate_CurrentPurity)

            CreatePet_TotalCostMutantMaterial_TextBlock.SetText(BasicMessage(GetMagnitudeLetter(CreatePet_TotalCost," x"," CREATE NEW SOURDOUGH")))

            NewPet:=GameManager.PlayersDataManager.CreatePet(pet_data_table{},CreatePet_PetToCreate_CurrentClass,CreatePet_PetToCreate_CurrentPurity,1,1)
            #
            RefreshPetStats(
                -1,
                array{},
                option{NewPet},
                CreatePet_TextureBlock_PetImages,

                CreatePet_Text_PackageA_LV,
                CreatePet_Text_PackageB_LV,
                
                CreatePet_TextBlock_Current_SpecialPerkA,
                CreatePet_TextBlock_Current_SpecialPerkB,
                CreatePet_TextBlock_Description_SpecialPerkA,
                CreatePet_TextBlock_Description_SpecialPerkB,
                CreatePet_TextBlock_Purity,
                CreatePet_TextBlock_Class,
                CreatePet_TextBlock_PackA,
                CreatePet_TextBlock_PackB,
                
                CreatePet_TextBlock_Current_TournamentHealth,
                CreatePet_TextBlock_Current_TournamentDamage,
                CreatePet_TextBlock_Current_TournamentSpeed,

                CreatePet_TextBlock_Current_PlayerPickDamage,
                CreatePet_TextBlock_Current_PlayerWeakPoint,

                CreatePet_TextBlock_Current_TournamentActionSpeed,
                CreatePet_TextBlock_Current_TournamentHealingPerk,
                CreatePet_TextBlock_Current_TournamentRange,

                CreatePet_TextBlock_Current_PlayerPickEfficiency,
                CreatePet_TextBlock_Current_PlayerPickCriticalDamage,

                array{},
                array{},
                array{},
                array{},
                array{},
                array{})
            #

            if(Agent := MaybeAgent?, Player:=player[Agent], Player.IsActive[], ValidBakery := GameManager.GetAgentBakery_Transacted(Agent)?):
                if(CreatePet_TotalCost < ValidBakery.PlayerDataCache.Inventory_MutantMaterial):
                    CreatePet_TotalCostMutantMaterial_TextBlock.SetTextColor(MutantMaterialTextColor)
                else:
                    CreatePet_TotalCostMutantMaterial_TextBlock.SetTextColor(NotEnoughTextColor)

                CreatePet_SlotsAvailable_TextBlock.SetText(BasicMessage("{ValidBakery.PlayerDataCache.AllPets.Length}/{ValidBakery.PlayerDataCache.MaxNoActive_PetSlots}"))
                if(ValidBakery.PlayerDataCache.AllPets.Length>=ValidBakery.PlayerDataCache.MaxNoActive_PetSlots):
                    CreatePet_SlotsAvailable_TextBlock.SetTextColor(Red)
                else:
                    CreatePet_SlotsAvailable_TextBlock.SetTextColor(White)
                CreatePet_TotalInventoryMutantMaterial_TextBlock.SetText(BasicMessage(GetMagnitudeLetter(ValidBakery.PlayerDataCache.Inventory_MutantMaterial,"","")))
                
    CreatePet_AddOrRemovePurity(Amount : int) : void =
        set CreatePet_PetToCreate_CurrentPurity = CreatePet_PetToCreate_CurrentPurity + Amount
        set ConfirmationToCreate = false
        CreatePet_TotalCostMutantMaterial_TextBlock.SetText(BasicMessage(GetMagnitudeLetter(CreatePet_TotalCost," x"," CREATE NEW SOURDOUGH")))
        if(CreatePet_PetToCreate_CurrentPurity<1):
            set CreatePet_PetToCreate_CurrentPurity = 1
        if(Agent := MaybeAgent?, Player := player[Agent], Player.IsActive[], ValidBakery := GameManager.GetAgentBakery_Transacted(Agent)?):
            if(ValidBakery.PlayerDataCache.PetTier<2, CreatePet_PetToCreate_CurrentPurity > 99):
                set CreatePet_PetToCreate_CurrentPurity = 99
                spawn:
                    RunPurityWarning()
            else if(ValidBakery.PlayerDataCache.PetTier<3, CreatePet_PetToCreate_CurrentPurity > 199):
                set CreatePet_PetToCreate_CurrentPurity = 199
            else if(CreatePet_PetToCreate_CurrentPurity > PetTotalMaxPurity):
                set CreatePet_PetToCreate_CurrentPurity = PetTotalMaxPurity
        CreatePet_RefreshAll()

    RunPurityWarning()<suspends> : void =
        CreatePet_TextBlock_Purity.SetText(BasicMessage("Max purity reached!"))
        CreatePet_TextBlock_Purity.SetTextColor(Red)
        Sleep(0.25)
        CreatePet_RefreshAll()
        CreatePet_TextBlock_Purity.SetTextColor(CurrentTextColor)
        Sleep(0.25)
        CreatePet_TextBlock_Purity.SetText(BasicMessage("Max purity reached!"))
        CreatePet_TextBlock_Purity.SetTextColor(Red)
        Sleep(0.25)
        CreatePet_TextBlock_Purity.SetTextColor(CurrentTextColor)
        CreatePet_RefreshAll()

    CreatePet_ChangeClass(Message : widget_message) : void =
        set CreatePet_PetToCreate_CurrentClass = CreatePet_PetToCreate_CurrentClass + 1
        set ConfirmationToCreate = false
        CreatePet_TotalCostMutantMaterial_TextBlock.SetText(BasicMessage(GetMagnitudeLetter(CreatePet_TotalCost," x"," CREATE NEW SOURDOUGH")))
        if(CreatePet_PetToCreate_CurrentClass>HUNTER):
            set CreatePet_PetToCreate_CurrentClass = 0
        CreatePet_RefreshAll()


    CreatePet_TryCreatePet(Message : widget_message) : void =
        if(ConfirmationToCreate?):
            set ConfirmationToCreate = false
            if(Message.Player.IsActive[], ValidBakery := GameManager.GetAgentBakery_Transacted(agent[Message.Player])?):
                if:
                    ValidBakery.PlayerDataCache.MaxNoActive_PetSlots>ValidBakery.PlayerDataCache.AllPets.Length #It is there less pet than max pet slots?
                then:
                    if:
                        Agent:=agent[Message.Player]
                        ValidBakery.PlayerDataCache.Inventory_MutantMaterial>=CreatePet_TotalCost 
                    then:
                        TableWithoutResource:=GameManager.PlayersDataManager.WithdrawResource(ValidBakery.PlayerDataCache,CreatePet_TotalCost,MUTANTMATERIAL)
                        NewPetTable:=GameManager.PlayersDataManager.CreatePet(pet_data_table{},CreatePet_PetToCreate_CurrentClass,CreatePet_PetToCreate_CurrentPurity,1,1)
                        AllCurrentUsedPets:=GetEquippedPets(TableWithoutResource)(0)
                            GameManager.VFXManager.PetCreatedObjetive_AudioPlayer.Play(Agent)
                        if(TableWithoutResource.MaxActive_PetSlots>AllCurrentUsedPets.Length):
                            UpdatePetTable:=GameManager.PlayersDataManager.SetEquippedPet(NewPetTable,true)
                            var PlayerDataTable : player_data_table = GameManager.PlayersDataManager.AddNewPet(TableWithoutResource,UpdatePetTable)
                            MaybeBakery:=GameManager.GetAgentBakery(Agent)
                            
                            #GameManager.PlayersDataManager.SaveNewData(Message.Player,PlayerData,FinalTable)
                            CreatePet_Remove_CanvasWithoutMessage()
                            
                            spawn:
                                RunPetCreationEffect(PlayerDataTable, UpdatePetTable) 
                            set PlayerDataTable = GameManager.QuestsAndOrdersDatabase.AddAgentQuest_ProgressV2(Message.Player,PlayerDataTable,1.0,CREATEPET_TUTORIAL_TYPE)(4)
                            ValidBakery.SetNewPlayerDataCache(PlayerDataTable)
            
                            spawn:
                                ValidBakery.GameManager.CalculateAndSaveFame(Agent)
                            Generate_PlayerEnhacements()
                        else:
                            UpdatePetTable:=GameManager.PlayersDataManager.SetEquippedPet(NewPetTable,false)
                            var PlayerDataTable : player_data_table = GameManager.PlayersDataManager.AddNewPet(TableWithoutResource,UpdatePetTable)
                            MaybeBakery:=GameManager.GetAgentBakery(Agent)
                            
                            #GameManager.PlayersDataManager.SaveNewData(Message.Player,PlayerData,FinalTable)    
                            #Slot slots for the new pet
                            spawn:
                                CreatePet_ShowNewPetOnHUD()
                            set PlayerDataTable = GameManager.QuestsAndOrdersDatabase.AddAgentQuest_ProgressV2(Message.Player,PlayerDataTable,1.0,CREATEPET_TUTORIAL_TYPE)(4)
                            ValidBakery.SetNewPlayerDataCache(PlayerDataTable)
                            Generate_PlayerEnhacements()
                        CreatePet_RefreshAll()
                        GameManager.RefreshInventoryUI(Agent)
                    else:
                        spawn:
                            CreatePet_RunError("CAN NOT AFFORD IT")
                else:
                    spawn:
                        CreatePet_RunError("NO SLOTS AVAILABLE")         
        else:
            CreatePet_TotalCostMutantMaterial_TextBlock.SetText(BasicMessage(GetMagnitudeLetter(CreatePet_TotalCost," x"," ARE YOU SURE?")))
            set ConfirmationToCreate = true

    CreatePet_ShowNewPetOnHUD()<suspends> : void =
        CreatePet_Remove_CanvasWithoutMessage()
        Sleep(0.1)
        EquipPet_Display_Canvas()

    GetEquippedPets(PlayerData : player_data_table) : tuple([]pet_data_table,[]int) =
        var AllPetsUsed : []pet_data_table = array{}
        var AllPetsUsedIndexes : []int = array{}
        for(Index:=0..PlayerData.AllPets.Length-1, Pet := PlayerData.AllPets[Index], Pet.IsEquipped?):
            set AllPetsUsed = AllPetsUsed + array{Pet}
            set AllPetsUsedIndexes = AllPetsUsedIndexes + array{Index}
        return (AllPetsUsed,AllPetsUsedIndexes)
        
    LoadEquippedPets(PlayerData : player_data_table) : void =
        #ResetAllPets
        for(Prop : AllPetsFollowProp):
            Prop.Dispose()

        set AllEquippedPetsIndexes = array{}
        set AllPetsFollowProp = array{}
        set AllPetsFollowProp_Tier = array{}

        #Now Load them
        for(PetIndex:=0..PlayerData.AllPets.Length-1, PetSelected := PlayerData.AllPets[PetIndex], PetSelected.IsEquipped?):
            set AllEquippedPetsIndexes = AllEquippedPetsIndexes + array{PetIndex}
            AssetProp:=GameManager.PropsManager.GetPetProp(PetSelected.Type)
            TuplePurity:=ConvertRawPurity(PetSelected.Purity)
            if(Agent:=MaybeAgent?, Character:=Agent.GetFortCharacter[]):
                ResultTuple:=SpawnProp(AssetProp,Character.GetTransform())
                if(ValidProp := ResultTuple(0)?):
                    SetStar(ValidProp,TuplePurity(0))
                    RemoveHealthUI(ValidProp)
                    set AllPetsFollowProp = AllPetsFollowProp + array{ValidProp}
                    set AllPetsFollowProp_Tier = AllPetsFollowProp_Tier + array{TuplePurity(0)}
        #Set -1 spaces
        loop:
            if(AllEquippedPetsIndexes.Length<3):
                set AllEquippedPetsIndexes = AllEquippedPetsIndexes + array{-1}
            else:
                break
        Generate_PlayerEnhacements()

    RunPetCreationEffect(FinalTable : player_data_table, PetTable : pet_data_table)<suspends> : void =
        if(Agent:=MaybeAgent?, Character := Agent.GetFortCharacter[], Player := player[Agent]):
            AssetProp:=GameManager.PropsManager.GetPetProp(PetTable.Type)
            TuplePurity:=ConvertRawPurity(PetTable.Purity)
            GameManager.VFXManager.RunPetCreation(Agent,AssetProp,TuplePurity(0)) #First, the cinematic
            ResultTuple:=SpawnProp(AssetProp,Character.GetTransform())
                if(ValidProp := ResultTuple(0)?):
                    SetStar(ValidProp,TuplePurity(0))
                    RemoveHealthUI(ValidProp)
                    if(AllPetsFollowProp_Tier.Length<>AllPetsFollowProp.Length):
                        PrintFatalError("On RunPetCreationEffect: AllPetsFollowProp_Tier.Length and AllPetsFollowProp.Length are not equal")
                    set AllPetsFollowProp = AllPetsFollowProp + array{ValidProp}
                    set AllPetsFollowProp_Tier = AllPetsFollowProp_Tier + array{TuplePurity(0)}

    CreatePet_RunError(String : string)<suspends> : void =
        CreatePet_TotalCostMutantMaterial_TextBlock.SetTextColor(Black)
        CreatePet_TotalCostMutantMaterial_TextBlock.SetText(BasicMessage("ERROR!: {String}"))
        Sleep(1.0)
        CreatePet_TotalCostMutantMaterial_TextBlock.SetTextColor(MutantMaterialTextColor)
        CreatePet_TotalCostMutantMaterial_TextBlock.SetText(BasicMessage(GetMagnitudeLetter(CreatePet_TotalCost," x"," CREATE NEW SOURDOUGH")))

    CreatePet_ExitPetSelector(Message : widget_message) : void =
        CreatePet_Remove_Canvas(Message)

    CreatePet_Remove_Canvas(Message : widget_message) : void =
        set CreatePet_UIIsDisplayed = false
        if:
            Agent := MaybeAgent?
            Player:=player[Agent]
            PlayerUI:=GetPlayerUI[Player]
            Canvas := CreatePet_MaybeCanvas?
        then:
            PlayerUI.RemoveWidget(Canvas)
            CreatePet_RemoveDelay_OnRemove()

    CreatePet_Remove_CanvasWithoutMessage() : void =
        set CreatePet_UIIsDisplayed = false
        if:
            Agent := MaybeAgent?
            Player:=player[Agent]
            PlayerUI:=GetPlayerUI[Player]
            Canvas := CreatePet_MaybeCanvas?
        then:
            PlayerUI.RemoveWidget(Canvas)
            CreatePet_RemoveDelay_OnRemove()

    CreatePet_RemoveDelay_OnAdd() : void =
        CreatePet_CloseButton.SetText(BasicMessage("Close"))

    CreatePet_RemoveDelay_OnRemove() : void =
        CreatePet_CloseButton.SetText(BasicMessage("Close "))
    
    CreatePet_Display_Canvas() : void =
        #Reset all selected pets
        if(not CreatePet_UIIsDisplayed?):
            set ConfirmationToCreate = false
            set FusionedPet = pet_data_table{}
            set IsFusionReady = false
            set AllMixPets_PetsToMix = array{-1,-1}
            set CreatePet_UIIsDisplayed = true
            var Index : int = 0 
            if:
                Agent := MaybeAgent?
                PlayerUI:=GetPlayerUI[player[Agent]]
                Canvas := CreatePet_MaybeCanvas?
            then:
                PlayerUI.AddWidget(Canvas,player_ui_slot{ZOrder:=0, InputMode := ui_input_mode.All})
                CreatePet_RemoveDelay_OnAdd()
                CreatePet_RefreshAll()

    MakeCreatePetCanvas(): canvas=
        Canvas := canvas:
            Slots:=array:
                #Main overlay
                canvas_slot:
                    Anchors:=anchors{Minimum:=vector2{X:=0.5,Y:=0.5},Maximum:=vector2{X:=0.5,Y:=0.5}}
                    ZOrder:=0
                    SizeToContent:=true
                    Offsets:=margin{Left:=0.0,Top:=0.0,Right:=0.0,Bottom:=0.0} #PositionX, PositionY, SizeX, SizeY
                    Alignment:=vector2{X:=0.5,Y:=0.5}
                    Widget:=texture_block{DefaultDesiredSize:=vector2{X:=3000.0,Y:=1250.0},DefaultImage:=Images.Img_MainBackground} 
                canvas_slot:
                    Anchors:=anchors{Minimum:=vector2{X:=1.0,Y:=0.0},Maximum:=vector2{X:=1.0,Y:=0.0}}
                    ZOrder:=0
                    SizeToContent:=true
                    Offsets:=margin{Left:=-10.0,Top:=10.0,Right:=0.0,Bottom:=0.0} #PositionX, PositionY, SizeX, SizeY
                    Alignment:=vector2{X:=1.0,Y:=0.0}
                    Widget:=stack_box:  
                        Orientation:=orientation.Vertical
                            Slots:=array:
                                stack_box_slot:
                                    HorizontalAlignment:=horizontal_alignment.Right
                                    VerticalAlignment:=vertical_alignment.Center
                                    Widget:=overlay:
                                        Slots:=array: 
                                            overlay_slot:
                                                Widget:=texture_block{DefaultDesiredSize:=vector2{X:=SizeBoxSaphiresX,Y:=SizeBoxSaphiresY},DefaultImage:=Images.EmptyTexture} 
                                            overlay_slot:
                                                HorizontalAlignment:=horizontal_alignment.Fill
                                                VerticalAlignment:=vertical_alignment.Fill
                                                Widget:=texture_block{DefaultImage:=Images.Img_LineBackground}  
                                            overlay_slot:
                                                HorizontalAlignment:=horizontal_alignment.Left
                                                VerticalAlignment:=vertical_alignment.Center
                                                Padding:=margin{Left:=Icon_LeftMargin}
                                                Widget:=texture_block{DefaultDesiredSize:=vector2{X:=IconSmallSizeX,Y:=IconSmallSizeY},DefaultImage:=Images.Mutant_Material_Icon}
                                            overlay_slot:
                                                HorizontalAlignment:=horizontal_alignment.Left
                                                VerticalAlignment:=vertical_alignment.Center
                                                Padding:=margin{Left:=TextIcon_LeftMargin}
                                                Widget:=CreatePet_TotalInventoryMutantMaterial_TextBlock    
                canvas_slot:
                    Anchors:=anchors{Minimum:=vector2{X:=0.5,Y:=0.5},Maximum:=vector2{X:=0.5,Y:=0.5}}
                    ZOrder:=0
                    SizeToContent:=true
                    Offsets:=margin{Left:=0.0,Top:=0.0,Right:=0.0,Bottom:=0.0} #PositionX, PositionY, SizeX, SizeY
                    Alignment:=vector2{X:=0.5,Y:=0.5}
                    Widget:=stack_box:  
                        Orientation:=orientation.Vertical
                            Slots:=array:
                                stack_box_slot: 
                                    HorizontalAlignment:=horizontal_alignment.Center
                                    VerticalAlignment:=vertical_alignment.Center
                                    Padding:=RowPadding
                                    Widget:=stack_box:  
                                        Orientation:=orientation.Horizontal
                                            Slots:=array:
                                                stack_box_slot: #SELECTED
                                                    Padding:=margin{Right:=SuperLowPadding,Left:=SuperLowPadding}
                                                    HorizontalAlignment:=horizontal_alignment.Center
                                                    VerticalAlignment:=vertical_alignment.Bottom
                                                    Widget:=stack_box:  
                                                        Orientation:=orientation.Vertical
                                                        Slots:=array:
                                                            stack_box_slot: 
                                                                Padding:=margin{Top:=SelectorBoxPadding,Bottom:=SelectorBoxPadding}
                                                                Widget:=overlay:
                                                                    Slots:=array:
                                                                        overlay_slot:
                                                                            HorizontalAlignment:=horizontal_alignment.Fill
                                                                            VerticalAlignment:=vertical_alignment.Fill
                                                                            Widget:=texture_block{DefaultImage:=Images.Img_LineBackground}
                                                                        overlay_slot:
                                                                            HorizontalAlignment:=horizontal_alignment.Fill
                                                                            VerticalAlignment:=vertical_alignment.Fill
                                                                            Padding:=margin{Left:=SelectorBoxPadding,Right:=SelectorBoxPadding,Top:=SelectorBoxPadding,Bottom:=SelectorBoxPadding}
                                                                            Widget:=texture_block{DefaultImage:=Images.Img_LineBackground}
                                                                        overlay_slot:
                                                                            HorizontalAlignment:=horizontal_alignment.Center
                                                                            VerticalAlignment:=vertical_alignment.Center
                                                                            Widget:=texture_block{DefaultImage:=Images.EmptyTexture,DefaultDesiredSize:=vector2{X:=950.0,Y:=600.0}}
                                                                        overlay_slot:
                                                                            HorizontalAlignment:=horizontal_alignment.Center
                                                                            VerticalAlignment:=vertical_alignment.Center
                                                                            Widget:=stack_box:  
                                                                                Orientation:=orientation.Vertical
                                                                                Slots:=array:
                                                                                    stack_box_slot: 
                                                                                        Widget:=stack_box:  
                                                                                            Orientation:=orientation.Horizontal
                                                                                            Slots:=array:
                                                                                                stack_box_slot: 
                                                                                                    Padding:=RowPadding
                                                                                                    Widget:=CreatePet_TextureBlock_PetImages
                                                                                                stack_box_slot: 
                                                                                                    HorizontalAlignment:=horizontal_alignment.Center
                                                                                                    VerticalAlignment:=vertical_alignment.Center
                                                                                                    Widget:=stack_box:  
                                                                                                        Orientation:=orientation.Vertical
                                                                                                        Slots:=array:
                                                                                                            stack_box_slot: 
                                                                                                                HorizontalAlignment:=horizontal_alignment.Center
                                                                                                                VerticalAlignment:=vertical_alignment.Center
                                                                                                                #Padding:=margin{Left:=TextPetsBoxPadding}
                                                                                                                Widget:=stack_box:
                                                                                                                    Orientation:=orientation.Horizontal
                                                                                                                    Slots:=array:
                                                                                                                        stack_box_slot: 
                                                                                                                            Padding:=margin{Left:=2.5,Right:=2.5}
                                                                                                                            Widget:=CreatePet_TextBlock_Class
                                                                                                                        stack_box_slot: 
                                                                                                                            Padding:=margin{Left:=2.5, Right:=1.0}
                                                                                                                            Widget:=CreatePet_TextBlock_PackA
                                                                                                                        stack_box_slot: 
                                                                                                                            Padding:=margin{Right:=2.5}
                                                                                                                            Widget:=CreatePet_TextBlock_PackB
                                                                                                            stack_box_slot: 
                                                                                                                Padding:=RowPadding
                                                                                                                Widget:=CreatePet_TextBlock_Purity
                                                                                    stack_box_slot: 
                                                                                        HorizontalAlignment:=horizontal_alignment.Fill
                                                                                        VerticalAlignment:=vertical_alignment.Fill
                                                                                        Widget:=texture_block{DefaultImage:=Images.Img_HorizontalLine,DefaultDesiredSize:=vector2{Y:=3.0}}
                                                                                    stack_box_slot:  
                                                                                        HorizontalAlignment:=horizontal_alignment.Center
                                                                                        VerticalAlignment:=vertical_alignment.Center
                                                                                        Padding:=RowPadding
                                                                                        Widget:=CreatePet_Text_PackageA_LV
                                                                                    stack_box_slot: 
                                                                                        HorizontalAlignment:=horizontal_alignment.Fill
                                                                                        VerticalAlignment:=vertical_alignment.Fill
                                                                                        Widget:=texture_block{DefaultImage:=Images.Img_HorizontalLine,DefaultDesiredSize:=vector2{Y:=3.0}}
                                                                                    stack_box_slot: 
                                                                                        HorizontalAlignment:=horizontal_alignment.Fill
                                                                                        VerticalAlignment:=vertical_alignment.Fill
                                                                                        Widget:=stack_box:  
                                                                                            Orientation:=orientation.Horizontal
                                                                                            Slots:=array:
                                                                                                stack_box_slot: 
                                                                                                    Widget:=stack_box:
                                                                                                        Orientation:=orientation.Vertical
                                                                                                        Slots:=array:
                                                                                                            stack_box_slot: 
                                                                                                                Padding:=RowPadding
                                                                                                                Widget:=text_block{DefaultText:=BasicMessage("Tournament stats"), DefaultTextColor:=Red, DefaultShadowOffset:=ShadowOffset, DefaultShadowOpacity := DefaultTextShadowOpacity}
                                                                                                            stack_box_slot: 
                                                                                                                HorizontalAlignment:=horizontal_alignment.Fill
                                                                                                                VerticalAlignment:=vertical_alignment.Fill
                                                                                                                Widget:=stack_box:  
                                                                                                                    Orientation:=orientation.Horizontal
                                                                                                                    Slots:=array:
                                                                                                                        stack_box_slot: 
                                                                                                                            HorizontalAlignment:=horizontal_alignment.Fill
                                                                                                                            VerticalAlignment:=vertical_alignment.Fill
                                                                                                                            Widget:=stack_box:  
                                                                                                                                Orientation:=orientation.Vertical
                                                                                                                                Slots:=array:
                                                                                                                                    stack_box_slot: 
                                                                                                                                        Widget:=overlay:
                                                                                                                                            Slots:=array:
                                                                                                                                                overlay_slot:
                                                                                                                                                    HorizontalAlignment:=horizontal_alignment.Left
                                                                                                                                                    Padding:=RowPadding
                                                                                                                                                    Widget:=text_block{DefaultText:=BasicMessage("Health:"), DefaultTextColor:=White, DefaultShadowOffset:=ShadowOffset, DefaultShadowOpacity := DefaultTextShadowOpacity}
                                                                                                                                                overlay_slot:
                                                                                                                                                    Widget:=texture_block{DefaultImage:=Images.EmptyTexture,DefaultDesiredSize:=vector2{X:=230.0,Y:=32.0}} 
                                                                                                                                    stack_box_slot: 
                                                                                                                                        HorizontalAlignment:=horizontal_alignment.Left
                                                                                                                                        Padding:=RowPadding
                                                                                                                                        Widget:=text_block{DefaultText:=BasicMessage("Damage:"), DefaultTextColor:=White, DefaultShadowOffset:=ShadowOffset, DefaultShadowOpacity := DefaultTextShadowOpacity}
                                                                                                                                    stack_box_slot: 
                                                                                                                                        HorizontalAlignment:=horizontal_alignment.Left
                                                                                                                                        Padding:=RowPadding
                                                                                                                                        Widget:=text_block{DefaultText:=BasicMessage("Mov. speed:"), DefaultTextColor:=White, DefaultShadowOffset:=ShadowOffset, DefaultShadowOpacity := DefaultTextShadowOpacity}
                                                                                                                        stack_box_slot: 
                                                                                                                            HorizontalAlignment:=horizontal_alignment.Fill
                                                                                                                            VerticalAlignment:=vertical_alignment.Fill
                                                                                                                            Widget:=stack_box:  
                                                                                                                                Orientation:=orientation.Vertical
                                                                                                                                Slots:=array:
                                                                                                                                    stack_box_slot: 
                                                                                                                                        Widget:=overlay:
                                                                                                                                            Slots:=array:
                                                                                                                                                overlay_slot:
                                                                                                                                                    Padding:=RowPadding
                                                                                                                                                    Widget:=CreatePet_TextBlock_Current_TournamentHealth
                                                                                                                                                overlay_slot:
                                                                                                                                                    Widget:=texture_block{DefaultImage:=Images.EmptyTexture,DefaultDesiredSize:=vector2{X:=257.0,Y:=32.0}}
                                                                                                                                    stack_box_slot: 
                                                                                                                                        Padding:=RowPadding
                                                                                                                                        Widget:=CreatePet_TextBlock_Current_TournamentDamage
                                                                                                                                    stack_box_slot: 
                                                                                                                                        Padding:=RowPadding
                                                                                                                                        Widget:=CreatePet_TextBlock_Current_TournamentSpeed
                                                                                                stack_box_slot: 
                                                                                                    HorizontalAlignment:=horizontal_alignment.Center
                                                                                                    VerticalAlignment:=vertical_alignment.Fill
                                                                                                    Widget:=texture_block{DefaultImage:=Images.Img_VerticalLine,DefaultDesiredSize:=vector2{X:=10.0}}
                                                                                                stack_box_slot: 
                                                                                                    Widget:=stack_box:
                                                                                                        Orientation:=orientation.Vertical
                                                                                                        Slots:=array:
                                                                                                            stack_box_slot: 
                                                                                                                Padding:=RowPadding
                                                                                                                Widget:=text_block{DefaultText:=BasicMessage("Player enhancements"), DefaultTextColor:=Cyan, DefaultShadowOffset:=ShadowOffset, DefaultShadowOpacity := DefaultTextShadowOpacity}
                                                                                                            stack_box_slot: 
                                                                                                                HorizontalAlignment:=horizontal_alignment.Fill
                                                                                                                VerticalAlignment:=vertical_alignment.Fill
                                                                                                                Widget:=stack_box:  
                                                                                                                    Orientation:=orientation.Horizontal
                                                                                                                    Slots:=array:
                                                                                                                        stack_box_slot: 
                                                                                                                            HorizontalAlignment:=horizontal_alignment.Fill
                                                                                                                            VerticalAlignment:=vertical_alignment.Fill
                                                                                                                            Widget:=stack_box:  
                                                                                                                                Orientation:=orientation.Vertical
                                                                                                                                Slots:=array:
                                                                                                                                    stack_box_slot: 
                                                                                                                                        Widget:=overlay:
                                                                                                                                            Slots:=array:
                                                                                                                                                overlay_slot:
                                                                                                                                                    Padding:=RowPadding
                                                                                                                                                    HorizontalAlignment:=horizontal_alignment.Left
                                                                                                                                                    Widget:=CreatePet_TextBlock_Description_SpecialPerkA
                                                                                                                                                overlay_slot:
                                                                                                                                                    Widget:=texture_block{DefaultImage:=Images.EmptyTexture,DefaultDesiredSize:=vector2{X:=300.0,Y:=32.0}}
                                                                                                                                    stack_box_slot: 
                                                                                                                                        Padding:=RowPadding
                                                                                                                                        HorizontalAlignment:=horizontal_alignment.Left
                                                                                                                                        Widget:=text_block{DefaultText:=BasicMessage("Pick damage:"), DefaultTextColor:=White, DefaultShadowOffset:=ShadowOffset, DefaultShadowOpacity := DefaultTextShadowOpacity}
                                                                                                                                    stack_box_slot: 
                                                                                                                                        Padding:=RowPadding
                                                                                                                                        HorizontalAlignment:=horizontal_alignment.Left
                                                                                                                                        Widget:=text_block{DefaultText:=BasicMessage("Weak point mult:"), DefaultTextColor:=White, DefaultShadowOffset:=ShadowOffset, DefaultShadowOpacity := DefaultTextShadowOpacity}
                                                                                                                        stack_box_slot: 
                                                                                                                            HorizontalAlignment:=horizontal_alignment.Fill
                                                                                                                            VerticalAlignment:=vertical_alignment.Fill
                                                                                                                            Widget:=stack_box:  
                                                                                                                                Orientation:=orientation.Vertical
                                                                                                                                Slots:=array:
                                                                                                                                    stack_box_slot: 
                                                                                                                                        Widget:=overlay:
                                                                                                                                            Slots:=array:
                                                                                                                                                overlay_slot:
                                                                                                                                                    Padding:=RowPadding
                                                                                                                                                    Widget:=CreatePet_TextBlock_Current_SpecialPerkA
                                                                                                                                                overlay_slot:
                                                                                                                                                    Widget:=texture_block{DefaultImage:=Images.EmptyTexture,DefaultDesiredSize:=vector2{X:=180.0,Y:=32.0}}
                                                                                                                                    stack_box_slot: 
                                                                                                                                        Padding:=RowPadding
                                                                                                                                        Widget:=CreatePet_TextBlock_Current_PlayerPickDamage
                                                                                                                                    stack_box_slot:
                                                                                                                                        Padding:=RowPadding 
                                                                                                                                        Widget:=CreatePet_TextBlock_Current_PlayerWeakPoint
                                                                                    stack_box_slot: 
                                                                                        HorizontalAlignment:=horizontal_alignment.Fill
                                                                                        VerticalAlignment:=vertical_alignment.Fill
                                                                                        Widget:=texture_block{DefaultImage:=Images.Img_HorizontalLine,DefaultDesiredSize:=vector2{Y:=3.0}}
                                                                                    stack_box_slot:  
                                                                                        HorizontalAlignment:=horizontal_alignment.Center
                                                                                        VerticalAlignment:=vertical_alignment.Center
                                                                                        Padding:=RowPadding
                                                                                        Widget:=CreatePet_Text_PackageB_LV
                                                                                    stack_box_slot: 
                                                                                        HorizontalAlignment:=horizontal_alignment.Fill
                                                                                        VerticalAlignment:=vertical_alignment.Fill
                                                                                        Widget:=texture_block{DefaultImage:=Images.Img_HorizontalLine,DefaultDesiredSize:=vector2{Y:=3.0}}
                                                                                    stack_box_slot: 
                                                                                        HorizontalAlignment:=horizontal_alignment.Fill
                                                                                        VerticalAlignment:=vertical_alignment.Fill
                                                                                        Widget:=stack_box:  
                                                                                            Orientation:=orientation.Horizontal
                                                                                            Slots:=array:
                                                                                                stack_box_slot: 
                                                                                                    Widget:=stack_box:
                                                                                                        Orientation:=orientation.Vertical
                                                                                                        Slots:=array:
                                                                                                            stack_box_slot: 
                                                                                                                Padding:=RowPadding
                                                                                                                Widget:=text_block{DefaultText:=BasicMessage("Tournament stats"), DefaultTextColor:=Red, DefaultShadowOffset:=ShadowOffset, DefaultShadowOpacity := DefaultTextShadowOpacity}
                                                                                                            stack_box_slot: 
                                                                                                                HorizontalAlignment:=horizontal_alignment.Fill
                                                                                                                VerticalAlignment:=vertical_alignment.Fill
                                                                                                                Widget:=stack_box:  
                                                                                                                    Orientation:=orientation.Horizontal
                                                                                                                    Slots:=array:
                                                                                                                        stack_box_slot: 
                                                                                                                            HorizontalAlignment:=horizontal_alignment.Fill
                                                                                                                            VerticalAlignment:=vertical_alignment.Fill
                                                                                                                            Widget:=stack_box:  
                                                                                                                                Orientation:=orientation.Vertical
                                                                                                                                Slots:=array:
                                                                                                                                    stack_box_slot: 
                                                                                                                                        Widget:=overlay:
                                                                                                                                            Slots:=array:
                                                                                                                                                overlay_slot:
                                                                                                                                                    Padding:=RowPadding
                                                                                                                                                    HorizontalAlignment:=horizontal_alignment.Left
                                                                                                                                                    Widget:=text_block{DefaultText:=BasicMessage("Attack time:"), DefaultTextColor:=White, DefaultShadowOffset:=ShadowOffset, DefaultShadowOpacity := DefaultTextShadowOpacity}
                                                                                                                                                overlay_slot:
                                                                                                                                                    Widget:=texture_block{DefaultImage:=Images.EmptyTexture,DefaultDesiredSize:=vector2{X:=230.0,Y:=32.0}}
                                                                                                                                    stack_box_slot: 
                                                                                                                                        Padding:=RowPadding
                                                                                                                                        HorizontalAlignment:=horizontal_alignment.Left
                                                                                                                                        Widget:=text_block{DefaultText:=BasicMessage("Range:"), DefaultTextColor:=White, DefaultShadowOffset:=ShadowOffset, DefaultShadowOpacity := DefaultTextShadowOpacity}
                                                                                                                                    stack_box_slot: 
                                                                                                                                        Padding:=RowPadding
                                                                                                                                        HorizontalAlignment:=horizontal_alignment.Left
                                                                                                                                        Widget:=text_block{DefaultText:=BasicMessage("Healing perk:"), DefaultTextColor:=White, DefaultShadowOffset:=ShadowOffset, DefaultShadowOpacity := DefaultTextShadowOpacity}
                                                                                                                        stack_box_slot: 
                                                                                                                            HorizontalAlignment:=horizontal_alignment.Fill
                                                                                                                            VerticalAlignment:=vertical_alignment.Fill
                                                                                                                            Widget:=stack_box:  
                                                                                                                                Orientation:=orientation.Vertical
                                                                                                                                Slots:=array:
                                                                                                                                    stack_box_slot: 
                                                                                                                                        Widget:=overlay:
                                                                                                                                            Slots:=array:
                                                                                                                                                overlay_slot:
                                                                                                                                                    Padding:=RowPadding
                                                                                                                                                    Widget:=CreatePet_TextBlock_Current_TournamentActionSpeed
                                                                                                                                                overlay_slot:
                                                                                                                                                    Widget:=texture_block{DefaultImage:=Images.EmptyTexture,DefaultDesiredSize:=vector2{X:=257.0,Y:=32.0}}
                                                                                                                                    stack_box_slot: 
                                                                                                                                        Padding:=RowPadding
                                                                                                                                        Widget:=CreatePet_TextBlock_Current_TournamentRange
                                                                                                                                    stack_box_slot: 
                                                                                                                                        Padding:=RowPadding
                                                                                                                                        Widget:=CreatePet_TextBlock_Current_TournamentHealingPerk
                                                                                                stack_box_slot: 
                                                                                                    HorizontalAlignment:=horizontal_alignment.Center
                                                                                                    VerticalAlignment:=vertical_alignment.Fill
                                                                                                    Widget:=texture_block{DefaultImage:=Images.Img_VerticalLine,DefaultDesiredSize:=vector2{X:=10.0}}
                                                                                                stack_box_slot: 
                                                                                                    Widget:=stack_box:
                                                                                                        Orientation:=orientation.Vertical
                                                                                                        Slots:=array:
                                                                                                            stack_box_slot: 
                                                                                                                Padding:=RowPadding
                                                                                                                Widget:=text_block{DefaultText:=BasicMessage("Player enhancements"), DefaultTextColor:=Cyan, DefaultShadowOffset:=ShadowOffset, DefaultShadowOpacity := DefaultTextShadowOpacity}
                                                                                                            stack_box_slot: 
                                                                                                                HorizontalAlignment:=horizontal_alignment.Fill
                                                                                                                VerticalAlignment:=vertical_alignment.Fill
                                                                                                                Widget:=stack_box:  
                                                                                                                    Orientation:=orientation.Horizontal
                                                                                                                    Slots:=array:
                                                                                                                        stack_box_slot: 
                                                                                                                            HorizontalAlignment:=horizontal_alignment.Fill
                                                                                                                            VerticalAlignment:=vertical_alignment.Fill
                                                                                                                            Widget:=stack_box:  
                                                                                                                                Orientation:=orientation.Vertical
                                                                                                                                Slots:=array:
                                                                                                                                    stack_box_slot: 
                                                                                                                                        Widget:=overlay:
                                                                                                                                            Slots:=array:
                                                                                                                                                overlay_slot:
                                                                                                                                                    Padding:=RowPadding
                                                                                                                                                    HorizontalAlignment:=horizontal_alignment.Left
                                                                                                                                                    Widget:=CreatePet_TextBlock_Description_SpecialPerkB
                                                                                                                                                overlay_slot:
                                                                                                                                                    Widget:=texture_block{DefaultImage:=Images.EmptyTexture,DefaultDesiredSize:=vector2{X:=300.0,Y:=32.0}}
                                                                                                                                    stack_box_slot: 
                                                                                                                                        Padding:=RowPadding
                                                                                                                                        HorizontalAlignment:=horizontal_alignment.Left
                                                                                                                                        Widget:=text_block{DefaultText:=BasicMessage("Pick efficiency:"), DefaultTextColor:=White, DefaultShadowOffset:=ShadowOffset, DefaultShadowOpacity := DefaultTextShadowOpacity}
                                                                                                                                    stack_box_slot: 
                                                                                                                                        Padding:=RowPadding
                                                                                                                                        HorizontalAlignment:=horizontal_alignment.Left
                                                                                                                                        Widget:=text_block{DefaultText:=BasicMessage("Pick critical dam:"), DefaultTextColor:=White, DefaultShadowOffset:=ShadowOffset, DefaultShadowOpacity := DefaultTextShadowOpacity}
                                                                                                                        stack_box_slot: 
                                                                                                                            HorizontalAlignment:=horizontal_alignment.Fill
                                                                                                                            VerticalAlignment:=vertical_alignment.Fill
                                                                                                                            Widget:=stack_box:  
                                                                                                                                Orientation:=orientation.Vertical
                                                                                                                                Slots:=array:
                                                                                                                                    stack_box_slot: 
                                                                                                                                        Widget:=overlay:
                                                                                                                                            Slots:=array:
                                                                                                                                                overlay_slot:
                                                                                                                                                    Padding:=RowPadding
                                                                                                                                                    Widget:=CreatePet_TextBlock_Current_SpecialPerkB
                                                                                                                                                overlay_slot:
                                                                                                                                                    Widget:=texture_block{DefaultImage:=Images.EmptyTexture,DefaultDesiredSize:=vector2{X:=180.0,Y:=32.0}}
                                                                                                                                    stack_box_slot: 
                                                                                                                                        Padding:=RowPadding
                                                                                                                                        Widget:=CreatePet_TextBlock_Current_PlayerPickEfficiency
                                                                                                                                    stack_box_slot: 
                                                                                                                                        Padding:=RowPadding
                                                                                                                                        Widget:=CreatePet_TextBlock_Current_PlayerPickCriticalDamage
                                                
                                                stack_box_slot:
                                                    HorizontalAlignment:=horizontal_alignment.Center
                                                    VerticalAlignment:=vertical_alignment.Center
                                                    Padding:=margin{Top:=5.0,Bottom:=5.0,Right:=5.0,Left:=5.0}
                                                    Widget:=overlay:
                                                        Slots:=array:
                                                            overlay_slot:
                                                                HorizontalAlignment:=horizontal_alignment.Fill
                                                                VerticalAlignment:=vertical_alignment.Fill
                                                                Widget:=texture_block{DefaultImage:=Images.Img_LineBackground}
                                                            overlay_slot:
                                                                HorizontalAlignment:=horizontal_alignment.Fill
                                                                VerticalAlignment:=vertical_alignment.Fill
                                                                Padding:=margin{Left:=SelectorBoxPadding,Right:=SelectorBoxPadding,Top:=SelectorBoxPadding,Bottom:=SelectorBoxPadding}
                                                                Widget:=texture_block{DefaultImage:=Images.Img_LineBackground}
                                                            overlay_slot:
                                                                HorizontalAlignment:=horizontal_alignment.Center
                                                                VerticalAlignment:=vertical_alignment.Center
                                                                Widget:=texture_block{DefaultImage:=Images.EmptyTexture,DefaultDesiredSize:=vector2{X:=140.0,Y:=140.0}}
                                                            overlay_slot:
                                                                Widget:=stack_box:  
                                                                    Orientation:=orientation.Vertical
                                                                        Slots:=array:
                                                                            stack_box_slot:
                                                                                Widget:=CreatePet_SlotsAvailable_TextBlock
                                                                            stack_box_slot:
                                                                                Widget:=text_block{DefaultText:=BasicMessage("Slot"), DefaultTextColor:=White, DefaultShadowOffset:=ShadowOffset, DefaultShadowOpacity := DefaultTextShadowOpacity}
                                stack_box_slot: 
                                    HorizontalAlignment:=horizontal_alignment.Center
                                    VerticalAlignment:=vertical_alignment.Center
                                    Padding:=RowPadding
                                    Widget:=overlay:
                                        Slots:=array:
                                            overlay_slot:
                                                Widget:=texture_block{DefaultImage:=Images.EmptyTexture,DefaultDesiredSize:=vector2{X:=650.0,Y:=100.0}}
                                            overlay_slot:
                                                Widget:=overlay:
                                                    Slots:=array:
                                                        overlay_slot:
                                                            Widget:=texture_block{DefaultImage:=Images.EmptyTexture,DefaultDesiredSize:=vector2{X:=650.0,Y:=80.0}}
                                                        overlay_slot:
                                                            HorizontalAlignment:=horizontal_alignment.Center
                                                            Widget:=overlay:
                                                                Slots:=array:
                                                                    overlay_slot:
                                                                        HorizontalAlignment:=horizontal_alignment.Left
                                                                        VerticalAlignment:=vertical_alignment.Center
                                                                        Padding:=margin{Left:=Icon_LeftMargin}
                                                                        Widget:=texture_block{DefaultDesiredSize:=vector2{X:=IconSmallSizeX,Y:=IconSmallSizeY},DefaultImage:=Images.Mutant_Material_Icon}
                                                                    overlay_slot:
                                                                        HorizontalAlignment:=horizontal_alignment.Left
                                                                        VerticalAlignment:=vertical_alignment.Center
                                                                        Padding:=margin{Left:=TextIcon_LeftMargin}
                                                                        Widget:=CreatePet_TotalCostMutantMaterial_TextBlock
                                                        overlay_slot:
                                                            HorizontalAlignment:=horizontal_alignment.Fill
                                                            VerticalAlignment:=vertical_alignment.Fill
                                                            Widget:=CreatePet_CreatePetButton
                                stack_box_slot: 
                                    HorizontalAlignment:=horizontal_alignment.Fill
                                    VerticalAlignment:=vertical_alignment.Center
                                    Padding:=RowPadding
                                    Widget:=CreatePet_ChangeClassButton

                                stack_box_slot: 
                                    HorizontalAlignment:=horizontal_alignment.Center
                                    VerticalAlignment:=vertical_alignment.Center
                                    Padding:=RowPadding
                                    Widget:=stack_box:  
                                        Orientation:=orientation.Horizontal
                                            Slots:=array:
                                                stack_box_slot:
                                                    Padding:=RowPadding
                                                    Widget:=overlay:
                                                        Slots:=array:
                                                            overlay_slot:
                                                                Widget:=texture_block{DefaultImage:=Images.EmptyTexture,DefaultDesiredSize:=vector2{X:=244.0,Y:=3.0}}
                                                            overlay_slot:
                                                                HorizontalAlignment:=horizontal_alignment.Fill
                                                                Widget:=CreatePet_Remove25PurityButton
                                                stack_box_slot:
                                                    Padding:=RowPadding
                                                    Widget:=overlay:
                                                        Slots:=array:
                                                            overlay_slot:
                                                                Widget:=texture_block{DefaultImage:=Images.EmptyTexture,DefaultDesiredSize:=vector2{X:=244.0,Y:=3.0}}
                                                            overlay_slot:
                                                                HorizontalAlignment:=horizontal_alignment.Fill
                                                                Widget:=CreatePet_Remove5PurityButton
                                                stack_box_slot:
                                                    Padding:=RowPadding
                                                    Widget:=overlay:
                                                        Slots:=array:
                                                            overlay_slot:
                                                                Widget:=texture_block{DefaultImage:=Images.EmptyTexture,DefaultDesiredSize:=vector2{X:=244.0,Y:=3.0}}
                                                            overlay_slot:
                                                                HorizontalAlignment:=horizontal_alignment.Fill
                                                                Widget:=CreatePet_Remove1PurityButton
                                                stack_box_slot:
                                                    Padding:=RowPadding
                                                    Widget:=overlay:
                                                        Slots:=array:
                                                            overlay_slot:
                                                                Widget:=texture_block{DefaultImage:=Images.EmptyTexture,DefaultDesiredSize:=vector2{X:=244.0,Y:=3.0}}
                                                            overlay_slot:
                                                                HorizontalAlignment:=horizontal_alignment.Fill
                                                                Widget:=CreatePet_Add1PurityButton
                                                stack_box_slot:
                                                    Padding:=RowPadding
                                                    Widget:=overlay:
                                                        Slots:=array:
                                                            overlay_slot:
                                                                Widget:=texture_block{DefaultImage:=Images.EmptyTexture,DefaultDesiredSize:=vector2{X:=244.0,Y:=3.0}}
                                                            overlay_slot:
                                                                HorizontalAlignment:=horizontal_alignment.Fill
                                                                Widget:=CreatePet_Add5PurityButton
                                                stack_box_slot:
                                                    Padding:=RowPadding
                                                    Widget:=overlay:
                                                        Slots:=array:
                                                            overlay_slot:
                                                                Widget:=texture_block{DefaultImage:=Images.EmptyTexture,DefaultDesiredSize:=vector2{X:=244.0,Y:=3.0}}
                                                            overlay_slot:
                                                                HorizontalAlignment:=horizontal_alignment.Fill
                                                                Widget:=CreatePet_Add25PurityButton      
                                stack_box_slot: 
                                    Padding:=RowPadding
                                    HorizontalAlignment:=horizontal_alignment.Fill
                                    Widget:=CreatePet_CloseButton

    #BUYSLOT
    ##################
    BuySlots_RefreshAll() : void =
        if(Agent := MaybeAgent?,Player:=player[Agent], Player.IsActive[], ValidBakery := GameManager.GetAgentBakery_Transacted(agent[Player])?):
            BuySlot_TotalAggedSapphires_TextBlock.SetText(BasicMessage(GetMagnitudeLetter(ValidBakery.PlayerDataCache.Inventory_AgedSapphires,"","")))
            BuySlot_TotalMutantMaterial_TextBlock.SetText(BasicMessage(GetMagnitudeLetter(ValidBakery.PlayerDataCache.Inventory_MutantMaterial,"","")))

            AllUsedPets:=GetEquippedPets(ValidBakery.PlayerDataCache)(0)

            #Active slots
            BuySlot_ActiveSlots_Current_TextBlock.SetText(BasicMessage("{ValidBakery.PlayerDataCache.MaxActive_PetSlots}"))
            if(ValidBakery.PlayerDataCache.MaxActive_PetSlots<MaxActivePetSlots):
                BuySlot_ActiveSlots_Upgrade_TextBlock.SetTextColor(UpgradeTextColor)
                BuySlot_ActiveSlots_Upgrade_TextBlock.SetText(BasicMessage("+1"))

                NextCost:=MainFormula_PerkPetCost(STAT_TYPE_PET_PERK_MAXACTIVE,ValidBakery.PlayerDataCache.MaxActive_PetSlots+1)
                BuySlot_ActiveSlots_Cost_TextBlock.SetText(BasicMessage(GetMagnitudeLetter(NextCost,"x ","")))
                BuySlot_ActiveSlots_Cost_TextBlock.SetTextColor(AgedSapphiresTextColor)
                if(NextCost>ValidBakery.PlayerDataCache.Inventory_AgedSapphires):
                    BuySlot_ActiveSlots_Cost_TextBlock.SetTextColor(NotEnoughTextColor)
                else:
                    BuySlot_ActiveSlots_Cost_TextBlock.SetTextColor(AgedSapphiresTextColor)
            else:
                BuySlot_ActiveSlots_Upgrade_TextBlock.SetTextColor(MaxedColor)
                BuySlot_ActiveSlots_Upgrade_TextBlock.SetText(BasicMessage("MAXED"))
                BuySlot_ActiveSlots_Cost_TextBlock.SetTextColor(MaxedColor)
                BuySlot_ActiveSlots_Cost_TextBlock.SetText(BasicMessage("MAXED"))
            #Max slots
            BuySlot_MaxSlots_Current_TextBlock.SetText(BasicMessage("{ValidBakery.PlayerDataCache.MaxNoActive_PetSlots}"))
            if(ValidBakery.PlayerDataCache.MaxNoActive_PetSlots<MaxPetSlots):
                BuySlot_MaxSlots_Upgrade_TextBlock.SetTextColor(UpgradeTextColor)
                BuySlot_MaxSlots_Upgrade_TextBlock.SetText(BasicMessage("+1"))

                NextCost:=MainFormula_PerkPetCost(STAT_TYPE_PET_PERK_MAXSLOTS,ValidBakery.PlayerDataCache.MaxNoActive_PetSlots+1)
                BuySlot_MaxSlots_Cost_TextBlock.SetText(BasicMessage(GetMagnitudeLetter(NextCost,"x ","")))
                BuySlot_MaxSlots_Cost_TextBlock.SetTextColor(MutantMaterialTextColor)
                if(NextCost>ValidBakery.PlayerDataCache.Inventory_MutantMaterial):
                    BuySlot_MaxSlots_Cost_TextBlock.SetTextColor(NotEnoughTextColor)
                else:
                    BuySlot_MaxSlots_Cost_TextBlock.SetTextColor(MutantMaterialTextColor)
            else:
                BuySlot_MaxSlots_Upgrade_TextBlock.SetTextColor(MaxedColor)
                BuySlot_MaxSlots_Upgrade_TextBlock.SetText(BasicMessage("MAXED"))
                BuySlot_MaxSlots_Cost_TextBlock.SetTextColor(MaxedColor)
                BuySlot_MaxSlots_Cost_TextBlock.SetText(BasicMessage("MAXED"))
            #Max tier
            


            BuySlot_TierUnlock_Current_TextBlock.SetText(BasicMessage("Tier: {ValidBakery.PlayerDataCache.PetTier}"))
            CurrentTier:=ValidBakery.PlayerDataCache.PetTier
            if(CurrentTier=1):
                BuySlot_TierUnlock_Current_TextBlock.SetText(BasicMessage("⭐"))
            if(CurrentTier=2):
                BuySlot_TierUnlock_Current_TextBlock.SetText(BasicMessage("⭐⭐"))
            if(CurrentTier=3):
                BuySlot_TierUnlock_Current_TextBlock.SetText(BasicMessage("⭐⭐⭐"))
            

            if(ValidBakery.PlayerDataCache.PetTier<MaxPetTier):
                BuySlot_TierUnlock_Upgrade_TextBlock.SetTextColor(UpgradeTextColor)
                NextTier:=ValidBakery.PlayerDataCache.PetTier+1
                if(NextTier=1):
                    BuySlot_TierUnlock_Upgrade_TextBlock.SetText(BasicMessage("⭐"))
                if(NextTier=2):
                    BuySlot_TierUnlock_Upgrade_TextBlock.SetText(BasicMessage("⭐⭐"))
                if(NextTier=3):
                    BuySlot_TierUnlock_Upgrade_TextBlock.SetText(BasicMessage("⭐⭐⭐"))

                NextCost:=MainFormula_PerkPetCost(STAT_TYPE_PET_PERK_TIER,ValidBakery.PlayerDataCache.PetTier+1)
                if(NextCost>ValidBakery.PlayerDataCache.Inventory_AgedSapphires):
                    BuySlot_TierUnlock_Cost_TextBlock.SetTextColor(NotEnoughTextColor)
                else:
                    BuySlot_TierUnlock_Cost_TextBlock.SetTextColor(AgedSapphiresTextColor)
                BuySlot_TierUnlock_Cost_TextBlock.SetText(BasicMessage(GetMagnitudeLetter(NextCost,"x ","")))
            else:
                BuySlot_TierUnlock_Upgrade_TextBlock.SetTextColor(MaxedColor)
                BuySlot_TierUnlock_Upgrade_TextBlock.SetText(BasicMessage("MAXED"))
                BuySlot_TierUnlock_Cost_TextBlock.SetTextColor(MaxedColor)
                BuySlot_TierUnlock_Cost_TextBlock.SetText(BasicMessage("MAXED"))
            
    BuySlots_TryImproveActiveSlots(Message : widget_message) : void =
        if(Agent := MaybeAgent?,Player:=player[Agent], Player.IsActive[], ValidBakery := GameManager.GetAgentBakery_Transacted(agent[Player])?):
            NextCost:=MainFormula_PerkPetCost(STAT_TYPE_PET_PERK_MAXACTIVE,ValidBakery.PlayerDataCache.MaxActive_PetSlots+1)
            if(NextCost<=ValidBakery.PlayerDataCache.Inventory_AgedSapphires, (ValidBakery.PlayerDataCache.MaxActive_PetSlots+1)<=MaxActivePetSlots):
                GameManager.VFXManager.BigObjetive_AudioPlayer.Play(Agent)
                NewLevel:=ValidBakery.PlayerDataCache.MaxActive_PetSlots+1
                TableWithLevel:=GameManager.PlayersDataManager.SetPetPerk(ValidBakery.PlayerDataCache,NewLevel,STAT_TYPE_PET_PERK_MAXACTIVE)
                TableWithouResourcers:=GameManager.PlayersDataManager.WithdrawResource(TableWithLevel,NextCost,AGEDSAPPHIRES)
                #GameManager.PlayersDataManager.SaveNewData(Player,PlayerData,TableWithouResourcers)
                ValidBakery.SetNewPlayerDataCache(TableWithouResourcers)
                ValidBakery.RefreshAllUI()
            else:
                #Print("TODO: Not enough, you need: {NextCost} and you have: {ValidBakery.PlayerDataCache.Inventory_AgedSapphires}")
            BuySlots_RefreshAll()


    BuySlots_TryImproveMaxSlots(Message : widget_message) : void =
        if(Agent := MaybeAgent?,Player:=player[Agent], Player.IsActive[], ValidBakery := GameManager.GetAgentBakery_Transacted(agent[Player])?):
            NextCost:=MainFormula_PerkPetCost(STAT_TYPE_PET_PERK_MAXSLOTS,ValidBakery.PlayerDataCache.MaxNoActive_PetSlots+1)
            if(NextCost<=ValidBakery.PlayerDataCache.Inventory_MutantMaterial, (ValidBakery.PlayerDataCache.MaxNoActive_PetSlots+1)<=MaxPetSlots):
                GameManager.VFXManager.SmallObjetive_AudioPlayer.Play(Agent)
                NewLevel:=ValidBakery.PlayerDataCache.MaxNoActive_PetSlots+1
                TableWithLevel:=GameManager.PlayersDataManager.SetPetPerk(ValidBakery.PlayerDataCache,NewLevel,STAT_TYPE_PET_PERK_MAXSLOTS)
                TableWithouResourcers:=GameManager.PlayersDataManager.WithdrawResource(TableWithLevel,NextCost,MUTANTMATERIAL)
                MaybeBakery:=GameManager.GetAgentBakery(Agent)
                ValidBakery.SetNewPlayerDataCache(TableWithouResourcers)
                ValidBakery.RefreshAllUI()
            else:
                #Print("TODO: Not enough, you need: {NextCost} and you have: {ValidBakery.PlayerDataCache.Inventory_MutantMaterial}")
            BuySlots_RefreshAll()

    BuySlots_TryImproveMaxPetTier(Message : widget_message) : void =
        if(Agent := MaybeAgent?,Player:=player[Agent], Player.IsActive[], ValidBakery := GameManager.GetAgentBakery_Transacted(agent[Player])?):
            NextCost:=MainFormula_PerkPetCost(STAT_TYPE_PET_PERK_TIER,ValidBakery.PlayerDataCache.PetTier+1)
            if(NextCost<=ValidBakery.PlayerDataCache.Inventory_AgedSapphires, (ValidBakery.PlayerDataCache.PetTier+1)<=MaxPetTier):
                GameManager.VFXManager.BigObjetive_AudioPlayer.Play(Agent)
                #GameManager.VFXManager.SmallObjetive_AudioPlayer.Play(Agent)
                NewLevel:=ValidBakery.PlayerDataCache.PetTier+1
                TableWithLevel:=GameManager.PlayersDataManager.SetPetPerk(ValidBakery.PlayerDataCache,NewLevel,STAT_TYPE_PET_PERK_TIER)
                TableWithouResourcers:=GameManager.PlayersDataManager.WithdrawResource(TableWithLevel,NextCost,AGEDSAPPHIRES)
                #GameManager.PlayersDataManager.SaveNewData(Player,PlayerData,TableWithouResourcers)
                MaybeBakery:=GameManager.GetAgentBakery(Agent)
                ValidBakery.SetNewPlayerDataCache(TableWithouResourcers)
                ValidBakery.RefreshAllUI()
            else:
                #Print("TODO: Not enough, you need: {NextCost} and you have: {ValidBakery.PlayerDataCache.Inventory_AgedSapphires}")
            BuySlots_RefreshAll()


    BuySlots_Remove_Canvas(Message : widget_message) : void =
        set BuySlots_UIIsDisplayed = false
        if:
            Agent := MaybeAgent?
            Player:=player[Agent]
            PlayerUI:=GetPlayerUI[Player]
            Canvas := BuySlots_MaybeCanvas?
        then:
            PlayerUI.RemoveWidget(Canvas)
            BuySlots_RemoveDelay_OnRemove()

    BuySlots_Remove_CanvasWithoutMessage() : void =
        set BuySlots_UIIsDisplayed = false
        if:
            Agent := MaybeAgent?
            Player:=player[Agent]
            PlayerUI:=GetPlayerUI[Player]
            Canvas := BuySlots_MaybeCanvas?
        then:
            PlayerUI.RemoveWidget(Canvas)
            BuySlots_RemoveDelay_OnRemove()

    BuySlots_RemoveDelay_OnAdd() : void =
        BuySlot_Close_Button.SetText(BasicMessage("Close"))

    BuySlots_RemoveDelay_OnRemove() : void =
        BuySlot_Close_Button.SetText(BasicMessage("Close "))

    BuySlots_Display_Canvas() : void =
        #Reset all selected pets
        if(not BuySlots_UIIsDisplayed?):
            set ConfirmationToCreate = false
            set FusionedPet = pet_data_table{}
            set IsFusionReady = false
            set AllMixPets_PetsToMix = array{-1,-1}
            set BuySlots_UIIsDisplayed = true
            var Index : int = 0 
            if:
                Agent := MaybeAgent?
                PlayerUI:=GetPlayerUI[player[Agent]]
                Canvas := BuySlots_MaybeCanvas?
            then:
                PlayerUI.AddWidget(Canvas,player_ui_slot{ZOrder:=0, InputMode := ui_input_mode.All})
                BuySlots_RemoveDelay_OnAdd()
                BuySlots_RefreshAll()

    MakeBuySlotsCanvas(): canvas=
        Canvas := canvas:
            Slots:=array:
                #Background image
                canvas_slot:
                    Anchors:=anchors{Minimum:=vector2{X:=0.5,Y:=0.5},Maximum:=vector2{X:=0.5,Y:=0.5}}
                    ZOrder:=0
                    SizeToContent:=true
                    Offsets:=margin{Left:=0.0,Top:=0.0,Right:=0.0,Bottom:=0.0} #PositionX, PositionY, SizeX, SizeY
                    Alignment:=vector2{X:=0.5,Y:=0.5}
                    Widget:=texture_block{DefaultImage:=Images.BlackBackground,DefaultDesiredSize:=vector2{X:=3000.0,Y:=1500.0}}
                canvas_slot:
                    Anchors:=anchors{Minimum:=vector2{X:=0.5,Y:=0.5},Maximum:=vector2{X:=0.5,Y:=0.5}}
                    ZOrder:=0
                    Offsets:=margin{Left:=0.0,Top:=0.0,Right:=0.0,Bottom:=0.0} #PositionX, PositionY, SizeX, SizeY
                    Alignment:=vector2{X:=0.5,Y:=0.5}
                    SizeToContent:=true
                    Widget:=texture_block:
                        DefaultImage:=Images.Img_MainBackground
                        DefaultDesiredSize:=vector2{X:=1375.0,Y:=1100.0}
                #Main overlay
                canvas_slot:
                    Anchors:=anchors{Minimum:=vector2{X:=0.5,Y:=0.5},Maximum:=vector2{X:=0.5,Y:=0.5}}
                    ZOrder:=0
                    SizeToContent:=true
                    #Offsets:=margin{Left:=1920.0,Top:=0.0,Right:=0.0,Bottom:=1080.0} #PositionX, PositionY, SizeX, SizeY
                    Alignment:=vector2{X:=0.5,Y:=0.5}
                    Widget:=overlay:
                        Slots:=array:
                            overlay_slot:
                                HorizontalAlignment:=horizontal_alignment.Center
                                VerticalAlignment:=vertical_alignment.Center
                                Widget:=stack_box:
                                    Orientation:=orientation.Vertical
                                        Slots:=array:
                                            stack_box_slot: #Sapphires per produc
                                                HorizontalAlignment:=horizontal_alignment.Right
                                                Widget:=stack_box:
                                                    Orientation:=orientation.Horizontal
                                                        Slots:=array:
                                                            stack_box_slot: #Current AgedSapphires
                                                                Padding:=margin{Left:=InterColumnLeftPadding,Top:=InterColumnTopPadding,Right:=InterColumnRightPadding,Bottom:=InterColumnBottomPadding}
                                                                Widget:=overlay:
                                                                    Slots:=array:
                                                                        overlay_slot:
                                                                            Widget:=texture_block{DefaultDesiredSize:=vector2{X:=SizeBoxSaphiresX,Y:=SizeBoxSaphiresY},DefaultImage:=Images.Img_LineBackground}
                                                                        overlay_slot:
                                                                            HorizontalAlignment:=horizontal_alignment.Left
                                                                            VerticalAlignment:=vertical_alignment.Center 
                                                                            Padding:=margin{Left:=10.0,Top:=0.0,Right:=0.0,Bottom:=0.0}
                                                                            Widget:=texture_block{DefaultDesiredSize:=vector2{X:=IconSmallSizeX,Y:=IconSmallSizeY}, DefaultImage:=Images.AgedSaphire_Icon}
                                                                        overlay_slot: 
                                                                            HorizontalAlignment:=horizontal_alignment.Left
                                                                            VerticalAlignment:=vertical_alignment.Center 
                                                                            Padding:=margin{Left:=65.0,Top:=0.0,Right:=0.0,Bottom:=0.0}
                                                                            Widget:=BuySlot_TotalAggedSapphires_TextBlock
                                                            stack_box_slot: #Current MutantMaterial
                                                                Padding:=margin{Left:=InterColumnLeftPadding,Top:=InterColumnTopPadding,Right:=InterColumnRightPadding,Bottom:=InterColumnBottomPadding}
                                                                Widget:=overlay:
                                                                    Slots:=array:
                                                                        overlay_slot:
                                                                            Widget:=texture_block{DefaultDesiredSize:=vector2{X:=SizeBoxSaphiresX,Y:=SizeBoxSaphiresY},DefaultImage:=Images.Img_LineBackground}
                                                                        overlay_slot:
                                                                            HorizontalAlignment:=horizontal_alignment.Left
                                                                            VerticalAlignment:=vertical_alignment.Center 
                                                                            Padding:=margin{Left:=10.0,Top:=0.0,Right:=0.0,Bottom:=0.0}
                                                                            Widget:=texture_block{DefaultDesiredSize:=vector2{X:=IconSmallSizeX,Y:=IconSmallSizeY}, DefaultImage:=Images.Mutant_Material_Icon}
                                                                        overlay_slot: 
                                                                            HorizontalAlignment:=horizontal_alignment.Left
                                                                            VerticalAlignment:=vertical_alignment.Center 
                                                                            Padding:=margin{Left:=65.0,Top:=0.0,Right:=0.0,Bottom:=0.0}
                                                                            Widget:=BuySlot_TotalMutantMaterial_TextBlock
                                            stack_box_slot: #Backpack perks
                                                Widget:=stack_box:
                                                    Orientation:=orientation.Vertical
                                                        Slots:=array:
                                                            stack_box_slot: #Line 1: Title
                                                                Padding:=margin{Left:=RowLeftPadding,Top:=RowTopPadding,Right:=RowRightPadding,Bottom:=RowBottomPadding}
                                                                HorizontalAlignment:=horizontal_alignment.Fill
                                                                VerticalAlignment:=vertical_alignment.Bottom
                                                                Widget:=overlay:
                                                                    Slots:=array:
                                                                        overlay_slot: 
                                                                            HorizontalAlignment:=horizontal_alignment.Fill
                                                                            VerticalAlignment:=vertical_alignment.Bottom
                                                                            Widget:=texture_block{DefaultDesiredSize:=vector2{X:=HorizontalLineSizeX,Y:=HorizontalLineSizeY},DefaultImage:=Images.Img_HorizontalLine}
                                                                        overlay_slot: 
                                                                            HorizontalAlignment:=horizontal_alignment.Center
                                                                            VerticalAlignment:=vertical_alignment.Center 
                                                                            Widget:=text_block{DefaultText:= BasicMessage("Sourdoughs perks"), DefaultTextColor := White, DefaultShadowOffset:=ShadowOffset, DefaultShadowOpacity := DefaultTextShadowOpacity}
                                                                        overlay_slot: 
                                                                            HorizontalAlignment:=horizontal_alignment.Center
                                                                            VerticalAlignment:=vertical_alignment.Center 
                                                                            Widget:=texture_block{DefaultDesiredSize:=vector2{X:=10.0,Y:=LineSizeY},DefaultImage:=Images.EmptyTexture}
                                                            stack_box_slot: #Line 2: All titles
                                                                Padding:=margin{Left:=RowLeftPadding,Top:=RowTopPadding,Right:=RowRightPadding,Bottom:=RowBottomPadding}
                                                                    Widget:=overlay:
                                                                        Slots:=array:
                                                                            overlay_slot: 
                                                                                HorizontalAlignment:=horizontal_alignment.Fill
                                                                                VerticalAlignment:=vertical_alignment.Center
                                                                                Widget:=texture_block{DefaultDesiredSize:=vector2{X:=0.0,Y:=LineSizeY},DefaultImage:=Images.Img_LineBackground}
                                                                            overlay_slot: 
                                                                                Widget:=stack_box:
                                                                                    Orientation:=orientation.Horizontal
                                                                                        Slots:=array:
                                                                                            stack_box_slot: 
                                                                                                Padding:=margin{Left:=InterColumnLeftPadding,Top:=InterColumnTopPadding,Right:=InterColumnRightPadding,Bottom:=InterColumnBottomPadding}
                                                                                                Widget:=overlay:
                                                                                                    Slots:=array:
                                                                                                        overlay_slot:
                                                                                                            Widget:=text_block{DefaultText:=BasicMessage("Perk"), DefaultTextColor:=White, DefaultShadowOffset:=ShadowOffset, DefaultShadowOpacity := DefaultTextShadowOpacity}
                                                                                                        overlay_slot: 
                                                                                                            Widget:=texture_block{DefaultDesiredSize:=vector2{X:=BUColumn2SizeX,Y:=BUColumn2SizeY},DefaultImage:=Images.EmptyTexture}
                                                                                            stack_box_slot: 
                                                                                                Padding:=margin{Left:=InterColumnLeftPadding,Top:=InterColumnTopPadding,Right:=InterColumnRightPadding,Bottom:=InterColumnBottomPadding}
                                                                                                Widget:=overlay:
                                                                                                    Slots:=array:
                                                                                                        overlay_slot:
                                                                                                            Widget:=text_block{DefaultText:=BasicMessage("Current"), DefaultTextColor:=White, DefaultShadowOffset:=ShadowOffset, DefaultShadowOpacity := DefaultTextShadowOpacity}
                                                                                                        overlay_slot: 
                                                                                                            Widget:=texture_block{DefaultDesiredSize:=vector2{X:=BUColumn4SizeX,Y:=BUColumn4SizeY},DefaultImage:=Images.EmptyTexture}
                                                                                            stack_box_slot: 
                                                                                                Padding:=margin{Left:=InterColumnLeftPadding,Top:=InterColumnTopPadding,Right:=InterColumnRightPadding,Bottom:=InterColumnBottomPadding}
                                                                                                Widget:=overlay:
                                                                                                    Slots:=array:
                                                                                                        overlay_slot:
                                                                                                            Widget:=text_block{DefaultText:=BasicMessage("Upgrade"), DefaultTextColor:=White, DefaultShadowOffset:=ShadowOffset, DefaultShadowOpacity := DefaultTextShadowOpacity}
                                                                                                        overlay_slot: 
                                                                                                            Widget:=texture_block{DefaultDesiredSize:=vector2{X:=BUColumn5SizeX,Y:=BUColumn5SizeY},DefaultImage:=Images.EmptyTexture}
                                                                                            stack_box_slot: 
                                                                                                Padding:=margin{Left:=InterColumnLeftPadding,Top:=InterColumnTopPadding,Right:=InterColumnRightPadding,Bottom:=InterColumnBottomPadding}
                                                                                                Widget:=overlay:
                                                                                                    Slots:=array:
                                                                                                        overlay_slot:
                                                                                                            Widget:=text_block{DefaultText:=BasicMessage("Upgrade cost"), DefaultTextColor:=White, DefaultShadowOffset:=ShadowOffset, DefaultShadowOpacity := DefaultTextShadowOpacity}
                                                                                                        overlay_slot: 
                                                                                                            Widget:=texture_block{DefaultDesiredSize:=vector2{X:=BUColumn6SizeX,Y:=BUColumn6SizeY},DefaultImage:=Images.EmptyTexture}

                                                                                                            
                                                            stack_box_slot: #Line 3: 
                                                                Padding:=margin{Left:=RowLeftPadding,Top:=RowTopPadding,Right:=RowRightPadding,Bottom:=RowBottomPadding}
                                                                    Widget:=overlay:
                                                                        Slots:=array:
                                                                            overlay_slot: 
                                                                                HorizontalAlignment:=horizontal_alignment.Fill
                                                                                VerticalAlignment:=vertical_alignment.Center
                                                                                Widget:=texture_block{DefaultDesiredSize:=vector2{X:=0.0,Y:=LineSizeY},DefaultImage:=Images.Img_LineBackground}
                                                                            overlay_slot: 
                                                                                Widget:=stack_box:
                                                                                    Orientation:=orientation.Horizontal
                                                                                        Slots:=array:
                                                                                            stack_box_slot: 
                                                                                                Padding:=margin{Left:=InterColumnLeftPadding,Top:=InterColumnTopPadding,Right:=InterColumnRightPadding,Bottom:=InterColumnBottomPadding}
                                                                                                Widget:=overlay:
                                                                                                    Slots:=array:
                                                                                                        overlay_slot:
                                                                                                            Widget:=text_block{DefaultText:=BasicMessage("Max equipped slots"), DefaultTextColor:=White, DefaultShadowOffset:=ShadowOffset, DefaultShadowOpacity := DefaultTextShadowOpacity}
                                                                                                        overlay_slot: 
                                                                                                            Widget:=texture_block{DefaultDesiredSize:=vector2{X:=BUColumn2SizeX,Y:=BUColumn2SizeY},DefaultImage:=Images.EmptyTexture}
                                                                                            stack_box_slot: 
                                                                                                Padding:=margin{Left:=InterColumnLeftPadding,Top:=InterColumnTopPadding,Right:=InterColumnRightPadding,Bottom:=InterColumnBottomPadding}
                                                                                                Widget:=overlay:
                                                                                                    Slots:=array:
                                                                                                        overlay_slot:
                                                                                                            Widget:=BuySlot_ActiveSlots_Current_TextBlock
                                                                                                        overlay_slot: 
                                                                                                            Widget:=texture_block{DefaultDesiredSize:=vector2{X:=BUColumn4SizeX,Y:=BUColumn4SizeY},DefaultImage:=Images.EmptyTexture}
                                                                                            stack_box_slot: 
                                                                                                Padding:=margin{Left:=InterColumnLeftPadding,Top:=InterColumnTopPadding,Right:=InterColumnRightPadding,Bottom:=InterColumnBottomPadding}
                                                                                                Widget:=overlay:
                                                                                                    Slots:=array:
                                                                                                        overlay_slot:
                                                                                                            Widget:=BuySlot_ActiveSlots_Upgrade_TextBlock
                                                                                                        overlay_slot: 
                                                                                                            Widget:=texture_block{DefaultDesiredSize:=vector2{X:=BUColumn5SizeX,Y:=BUColumn5SizeY},DefaultImage:=Images.EmptyTexture}
                                                                                            stack_box_slot: 
                                                                                                Padding:=margin{Left:=InterColumnLeftPadding,Top:=InterColumnTopPadding,Right:=InterColumnRightPadding,Bottom:=InterColumnBottomPadding}
                                                                                                Widget:=overlay:
                                                                                                    Slots:=array:
                                                                                                        overlay_slot: 
                                                                                                            Widget:=texture_block{DefaultDesiredSize:=vector2{X:=BUColumn6SizeX,Y:=BUColumn6SizeY},DefaultImage:=Images.EmptyTexture}
                                                                                                        overlay_slot: 
                                                                                                            HorizontalAlignment:=horizontal_alignment.Left
                                                                                                            VerticalAlignment:=vertical_alignment.Center 
                                                                                                            Padding:=margin{Left:=65.0,Top:=0.0,Right:=0.0,Bottom:=0.0}
                                                                                                            Widget:= BuySlot_ActiveSlots_Cost_TextBlock
                                                                                                        overlay_slot: 
                                                                                                            HorizontalAlignment:=horizontal_alignment.Left
                                                                                                            VerticalAlignment:=vertical_alignment.Center 
                                                                                                            Padding:=margin{Left:=10.0,Top:=0.0,Right:=0.0,Bottom:=0.0}
                                                                                                            Widget:=BuySlot_ActiveSlots_ResourceIcon
                                                                                                        overlay_slot:
                                                                                                            HorizontalAlignment:=horizontal_alignment.Fill
                                                                                                            VerticalAlignment:=vertical_alignment.Center  
                                                                                                            Widget:=BuySlot_ActiveSlots_Button
                                                                                            

                                                            stack_box_slot: #Line 4: All titles
                                                                Padding:=margin{Left:=RowLeftPadding,Top:=RowTopPadding,Right:=RowRightPadding,Bottom:=RowBottomPadding}
                                                                    Widget:=overlay:
                                                                        Slots:=array:
                                                                            overlay_slot: 
                                                                                HorizontalAlignment:=horizontal_alignment.Fill
                                                                                VerticalAlignment:=vertical_alignment.Center
                                                                                Widget:=texture_block{DefaultDesiredSize:=vector2{X:=0.0,Y:=LineSizeY},DefaultImage:=Images.Img_LineBackground}
                                                                            overlay_slot: 
                                                                                Widget:=stack_box:
                                                                                    Orientation:=orientation.Horizontal
                                                                                        Slots:=array:
                                                                                            stack_box_slot: 
                                                                                                Padding:=margin{Left:=InterColumnLeftPadding,Top:=InterColumnTopPadding,Right:=InterColumnRightPadding,Bottom:=InterColumnBottomPadding}
                                                                                                Widget:=overlay:
                                                                                                    Slots:=array:
                                                                                                        overlay_slot:
                                                                                                            Widget:=text_block{DefaultText:=BasicMessage("Max inventory slots"), DefaultTextColor:=White, DefaultShadowOffset:=ShadowOffset, DefaultShadowOpacity := DefaultTextShadowOpacity}
                                                                                                        overlay_slot: 
                                                                                                            Widget:=texture_block{DefaultDesiredSize:=vector2{X:=BUColumn2SizeX,Y:=BUColumn2SizeY},DefaultImage:=Images.EmptyTexture}
                                                                                            stack_box_slot: 
                                                                                                Padding:=margin{Left:=InterColumnLeftPadding,Top:=InterColumnTopPadding,Right:=InterColumnRightPadding,Bottom:=InterColumnBottomPadding}
                                                                                                Widget:=overlay:
                                                                                                    Slots:=array:
                                                                                                        overlay_slot:
                                                                                                            Widget:=BuySlot_MaxSlots_Current_TextBlock
                                                                                                        overlay_slot: 
                                                                                                            Widget:=texture_block{DefaultDesiredSize:=vector2{X:=BUColumn4SizeX,Y:=BUColumn4SizeY},DefaultImage:=Images.EmptyTexture}
                                                                                            stack_box_slot: 
                                                                                                Padding:=margin{Left:=InterColumnLeftPadding,Top:=InterColumnTopPadding,Right:=InterColumnRightPadding,Bottom:=InterColumnBottomPadding}
                                                                                                Widget:=overlay:
                                                                                                    Slots:=array:
                                                                                                        overlay_slot:
                                                                                                            Widget:=BuySlot_MaxSlots_Upgrade_TextBlock
                                                                                                        overlay_slot: 
                                                                                                            Widget:=texture_block{DefaultDesiredSize:=vector2{X:=BUColumn5SizeX,Y:=BUColumn5SizeY},DefaultImage:=Images.EmptyTexture}
                                                                                            stack_box_slot: 
                                                                                                Padding:=margin{Left:=InterColumnLeftPadding,Top:=InterColumnTopPadding,Right:=InterColumnRightPadding,Bottom:=InterColumnBottomPadding}
                                                                                                Widget:=overlay:
                                                                                                    Slots:=array:
                                                                                                        overlay_slot: 
                                                                                                            Widget:=texture_block{DefaultDesiredSize:=vector2{X:=BUColumn6SizeX,Y:=BUColumn6SizeY},DefaultImage:=Images.EmptyTexture}
                                                                                                        overlay_slot: 
                                                                                                            HorizontalAlignment:=horizontal_alignment.Left
                                                                                                            VerticalAlignment:=vertical_alignment.Center 
                                                                                                            Padding:=margin{Left:=65.0,Top:=0.0,Right:=0.0,Bottom:=0.0}
                                                                                                            Widget:=BuySlot_MaxSlots_Cost_TextBlock 
                                                                                                        overlay_slot: 
                                                                                                            HorizontalAlignment:=horizontal_alignment.Left
                                                                                                            VerticalAlignment:=vertical_alignment.Center 
                                                                                                            Padding:=margin{Left:=10.0,Top:=0.0,Right:=0.0,Bottom:=0.0}
                                                                                                            Widget:=BuySlot_MaxSlots_ResourceIcon
                                                                                                        overlay_slot:
                                                                                                            HorizontalAlignment:=horizontal_alignment.Fill
                                                                                                            VerticalAlignment:=vertical_alignment.Center  
                                                                                                            Widget:=BuySlot_MaxSlots_Button
                                                                                                
                                                                            

                                                            stack_box_slot: #Line 5: All titles
                                                                Padding:=margin{Left:=RowLeftPadding,Top:=RowTopPadding,Right:=RowRightPadding,Bottom:=RowBottomPadding}
                                                                    Widget:=overlay:
                                                                        Slots:=array:
                                                                            overlay_slot: 
                                                                                HorizontalAlignment:=horizontal_alignment.Fill
                                                                                VerticalAlignment:=vertical_alignment.Center
                                                                                Widget:=texture_block{DefaultDesiredSize:=vector2{X:=0.0,Y:=LineSizeY},DefaultImage:=Images.Img_LineBackground}
                                                                            overlay_slot: 
                                                                                Widget:=stack_box:
                                                                                    Orientation:=orientation.Horizontal
                                                                                        Slots:=array:
                                                                                            stack_box_slot: 
                                                                                                Padding:=margin{Left:=InterColumnLeftPadding,Top:=InterColumnTopPadding,Right:=InterColumnRightPadding,Bottom:=InterColumnBottomPadding}
                                                                                                Widget:=overlay:
                                                                                                    Slots:=array:
                                                                                                        overlay_slot:
                                                                                                            Widget:=text_block{DefaultText:=BasicMessage("Unlock purity tier"), DefaultTextColor:=White, DefaultShadowOffset:=ShadowOffset, DefaultShadowOpacity := DefaultTextShadowOpacity}
                                                                                                        overlay_slot: 
                                                                                                            Widget:=texture_block{DefaultDesiredSize:=vector2{X:=BUColumn2SizeX,Y:=BUColumn2SizeY},DefaultImage:=Images.EmptyTexture}
                                                                                            stack_box_slot: 
                                                                                                Padding:=margin{Left:=InterColumnLeftPadding,Top:=InterColumnTopPadding,Right:=InterColumnRightPadding,Bottom:=InterColumnBottomPadding}
                                                                                                Widget:=overlay:
                                                                                                    Slots:=array:
                                                                                                        overlay_slot:
                                                                                                            Widget:=BuySlot_TierUnlock_Current_TextBlock
                                                                                                        overlay_slot: 
                                                                                                            Widget:=texture_block{DefaultDesiredSize:=vector2{X:=BUColumn4SizeX,Y:=BUColumn4SizeY},DefaultImage:=Images.EmptyTexture}
                                                                                            stack_box_slot: 
                                                                                                Padding:=margin{Left:=InterColumnLeftPadding,Top:=InterColumnTopPadding,Right:=InterColumnRightPadding,Bottom:=InterColumnBottomPadding}
                                                                                                Widget:=overlay:
                                                                                                    Slots:=array:
                                                                                                        overlay_slot:
                                                                                                            Widget:=BuySlot_TierUnlock_Upgrade_TextBlock
                                                                                                        overlay_slot: 
                                                                                                            Widget:=texture_block{DefaultDesiredSize:=vector2{X:=BUColumn5SizeX,Y:=BUColumn5SizeY},DefaultImage:=Images.EmptyTexture}
                                                                                            stack_box_slot: 
                                                                                                Padding:=margin{Left:=InterColumnLeftPadding,Top:=InterColumnTopPadding,Right:=InterColumnRightPadding,Bottom:=InterColumnBottomPadding}
                                                                                                Widget:=overlay:
                                                                                                    Slots:=array:
                                                                                                        overlay_slot: 
                                                                                                            Widget:=texture_block{DefaultDesiredSize:=vector2{X:=BUColumn6SizeX,Y:=BUColumn6SizeY},DefaultImage:=Images.EmptyTexture}
                                                                                                        overlay_slot: 
                                                                                                            HorizontalAlignment:=horizontal_alignment.Left
                                                                                                            VerticalAlignment:=vertical_alignment.Center 
                                                                                                            Padding:=margin{Left:=65.0,Top:=0.0,Right:=0.0,Bottom:=0.0}
                                                                                                            Widget:=BuySlot_TierUnlock_Cost_TextBlock 
                                                                                                        overlay_slot: 
                                                                                                            HorizontalAlignment:=horizontal_alignment.Left
                                                                                                            VerticalAlignment:=vertical_alignment.Center 
                                                                                                            Padding:=margin{Left:=10.0,Top:=0.0,Right:=0.0,Bottom:=0.0}
                                                                                                            Widget:=BuySlot_TierUnlock_ResourceIcon
                                                                                                        overlay_slot:
                                                                                                            HorizontalAlignment:=horizontal_alignment.Fill
                                                                                                            VerticalAlignment:=vertical_alignment.Center  
                                                                                                            Widget:=BuySlot_TierUnlock_Button
                                                                                                        
                                                            stack_box_slot: #Line 8: All titles
                                                                Padding:=margin{Left:=RowLeftPadding,Top:=RowTopPadding,Right:=RowRightPadding,Bottom:=RowBottomPadding}
                                                                    Widget:=overlay:
                                                                        Slots:=array:
                                                                            overlay_slot: 
                                                                                HorizontalAlignment:=horizontal_alignment.Center
                                                                                VerticalAlignment:=vertical_alignment.Center
                                                                                Widget:=stack_box:
                                                                                    Orientation:=orientation.Horizontal
                                                                                        Slots:=array:
                                                                                            stack_box_slot: 
                                                                                                Padding:=margin{Left:=InterColumnLeftPadding,Top:=InterColumnTopPadding,Right:=InterColumnRightPadding,Bottom:=InterColumnBottomPadding}
                                                                                                Widget:=overlay:
                                                                                                    Slots:=array:
                                                                                                        overlay_slot:
                                                                                                            Widget:=texture_block{DefaultImage:=Images.EmptyTexture,DefaultDesiredSize:=vector2{X:=BigCloseButtonSizeX,Y:=80.0}}
                                                                                                        overlay_slot: 
                                                                                                            HorizontalAlignment:=horizontal_alignment.Fill
                                                                                                            VerticalAlignment:=vertical_alignment.Fill
                                                                                                            Padding:=margin{Left:=InterColumnLeftPadding,Top:=InterColumnTopPadding,Right:=InterColumnRightPadding,Bottom:=InterColumnBottomPadding}
                                                                                                            Widget:=BuySlot_Close_Button


    #RECICLE PET
    ##################
    RecyclePet_ExecuteRecycle(Message : widget_message)<suspends> : void =
        if(ConfirmationToRecycle<=0, IsRecycleReady?):
            RecyclePet_ExecuteRecycleButton.SetText(BasicMessage("ARE YOU SURE?"))
            set ConfirmationToRecycle = ConfirmationToRecycle + 1
            set IsRecycleReady = true
        else if(ConfirmationToRecycle=1, IsRecycleReady?):
            RecyclePet_ExecuteRecycleButton.SetText(BasicMessage("SURE? LAST WARNING"))
            set ConfirmationToRecycle = ConfirmationToRecycle + 1
            set IsRecycleReady = true
        else if(ConfirmationToRecycle>=2, IsRecycleReady?):
            set IsRecycleReady = false
            if(Message.Player.IsActive[], ValidBakery := GameManager.GetAgentBakery_Transacted(agent[Message.Player])?):
                AllPets:=ValidBakery.PlayerDataCache.AllPets
                if:
                    PetToRecycle:=ValidBakery.PlayerDataCache.AllPets[RecyclePet_CurrentOption]
                    NewAllPets:=AllPets.RemoveElement[RecyclePet_CurrentOption]
                    Agent := agent[Message.Player]
                then:
                    TupleRecycleAmount:=CalculePetTotalAmountsInvested(PetToRecycle)
                    GrantedMutantMaterial:=TupleRecycleAmount(0)*RecycleEfficiency
                    GrantedCoal:=(TupleRecycleAmount(1)+TupleRecycleAmount(2))*RecycleEfficiency

                    PlayerTableWithoutPet:=GameManager.PlayersDataManager.SavePets(ValidBakery.PlayerDataCache,NewAllPets)
                    PlayerTableWithMutantMaterial:=GameManager.PlayersDataManager.AddResourcers(PlayerTableWithoutPet,GrantedMutantMaterial,MUTANTMATERIAL,-1.0)
                    PlayerTableWithSapphires:=GameManager.PlayersDataManager.AddResourcers(PlayerTableWithMutantMaterial(0),GrantedCoal,COAL,-1.0)
                    #GameManager.PlayersDataManager.SaveNewData(Message.Player,PlayerData,PlayerTableWithSapphires(0))
                    MaybeBakery:=GameManager.GetAgentBakery(Agent)
                    ValidBakery.SetNewPlayerDataCache(PlayerTableWithSapphires(0))
                    spawn:
                        ValidBakery.GameManager.CalculateAndSaveFame(Agent)
                    GameManager.RefreshInventoryUI(Agent)
                    spawn:
                        HandleRecyclePetsEffect()
                    if(PlayerTableWithSapphires(0).AllPets.Length>0):
                        RecyclePet_RefreshAll()
                    else:
                        Handler_FromEverything_To_Menu(Message, SUBMENU_RECYCLE_PET)
                    LoadEquippedPets(PlayerTableWithSapphires(0))
            set ConfirmationToRecycle = 0
            set IsRecycleReady = true
        else:
            RecyclePet_ExecuteRecycleButton.SetText(BasicMessage("RECYCLE NOT READY!"))

    HandlerExecuteRecycle(Message : widget_message) : void =
        if(IsRecycleReady?):
            spawn:
                RecyclePet_ExecuteRecycle(Message)

    HandleRecyclePetsEffect()<suspends> : void =
        var TotalProgress : int = 0
        loop:
            set TotalProgress = TotalProgress + 5
            RecyclePet_ExecuteRecycleButton.SetText(BasicMessage("RECYCLING {TotalProgress}%"))
            Sleep(0.04)   
            if(TotalProgress>=100):
                RecyclePet_ExecuteRecycleButton.SetText(BasicMessage("SUCCESS!"))
                break
        MixPets_RefreshAll()
        RecyclePet_ExecuteRecycleButton.SetText(BasicMessage("RECYCLE"))

    RecyclePet_NavidateRight(Message : widget_message) : void =
        if(Message.Player.IsActive[], ValidBakery := GameManager.GetAgentBakery_Transacted(agent[Message.Player])?):
            set RecyclePet_CurrentOption = RecyclePet_CurrentOption + 1       
            if(RecyclePet_CurrentOption > ValidBakery.PlayerDataCache.AllPets.Length-1):
                set RecyclePet_CurrentOption = ValidBakery.PlayerDataCache.AllPets.Length-1
        spawn:
            RecyclePet_RefreshAll()

    RecyclePet_NavidateLeft(Message : widget_message) : void =
        set RecyclePet_CurrentOption = RecyclePet_CurrentOption - 1
        if(RecyclePet_CurrentOption < 0):
            set RecyclePet_CurrentOption = 0
        spawn:
            RecyclePet_RefreshAll()


    RecyclePet_Remove_Canvas(Message : widget_message) : void =
        set RecyclePet_UIIsDisplayed = false
        if:
            Agent := MaybeAgent?
            Player:=player[Agent]
            PlayerUI:=GetPlayerUI[Player]
            Canvas := RecyclePet_MaybeCanvas?
        then:
            PlayerUI.RemoveWidget(Canvas)
            RecyclePet_RemoveDelay_OnRemove()
            spawn:
                RecyclePet_RefreshAll()

    RecyclePet_Remove_CanvasWithoutMessage() : void =
        set RecyclePet_UIIsDisplayed = false
        if:
            Agent := MaybeAgent?
            Player:=player[Agent]
            PlayerUI:=GetPlayerUI[Player]
            Canvas := RecyclePet_MaybeCanvas?
        then:
            PlayerUI.RemoveWidget(Canvas)
            RecyclePet_RemoveDelay_OnRemove()
            spawn:
                RecyclePet_RefreshAll()

    RecyclePet_RemoveDelay_OnAdd() : void =
        RecyclePet_CancelButton.SetText(BasicMessage("Cancel "))

    RecyclePet_RemoveDelay_OnRemove() : void =
        RecyclePet_CancelButton.SetText(BasicMessage("Cancel "))

    RecyclePet_Display_Canvas() : void =
        #Reset all selected pets
        if(not RecyclePet_UIIsDisplayed?):
            set ConfirmationToCreate = false
            set RecyclePet_UIIsDisplayed = true
            var Index : int = 0 
            if:
                Agent := MaybeAgent?
                PlayerUI:=GetPlayerUI[player[Agent]]
                Canvas := RecyclePet_MaybeCanvas?
            then:
                PlayerUI.AddWidget(Canvas,player_ui_slot{ZOrder:=0, InputMode := ui_input_mode.All})
                RecyclePet_RemoveDelay_OnAdd()
                spawn:
                    RecyclePet_RefreshAll()

    RecyclePet_RefreshAll()<suspends> : void =
        set ConfirmationToRecycle = 0
        set IsRecycleReady = false
        RecyclePet_Text_MuntantMaterialRecicled.SetText(BasicMessage("-"))
        RecyclePet_Text_CoalRecicled.SetText(BasicMessage("-"))
        RecyclePet_ExecuteRecycleButton.SetText(BasicMessage("Calculating..."))
        if(Agent := MaybeAgent?, Player := player[Agent], Player.IsActive[], ValidBakery := GameManager.GetAgentBakery_Transacted(agent[Player])?):
            AllPets:=ValidBakery.PlayerDataCache.AllPets
            if(Pet := AllPets[RecyclePet_CurrentOption]):
                RefreshPetStats(
                    RecyclePet_CurrentOption,
                    ValidBakery.PlayerDataCache.AllPets,
                    option{Pet},
                    RecyclePet_Selected_TextureBlock_PetImages,

                    RecyclePet_Text_PackageA_LV,
                    RecyclePet_Text_PackageB_LV,
                    
                    RecyclePet_Selected_TextBlock_Current_SpecialPerkA,
                    RecyclePet_Selected_TextBlock_Current_SpecialPerkB,
                    RecyclePet_Selected_TextBlock_Description_SpecialPerkA,
                    RecyclePet_Selected_TextBlock_Description_SpecialPerkB,
                    RecyclePet_Selected_TextBlock_Purity,
                    RecyclePet_Selected_TextBlock_Class,
                    RecyclePet_Selected_TextBlock_PackA,
                    RecyclePet_Selected_TextBlock_PackB,
                    
                    RecyclePet_Selected_TextBlock_Current_TournamentHealth,
                    RecyclePet_Selected_TextBlock_Current_TournamentDamage,
                    RecyclePet_Selected_TextBlock_Current_TournamentSpeed,

                    RecyclePet_Selected_TextBlock_Current_PlayerPickDamage,
                    RecyclePet_Selected_TextBlock_Current_PlayerWeakPoint,

                    RecyclePet_Selected_TextBlock_Current_TournamentActionSpeed,
                    RecyclePet_Selected_TextBlock_Current_TournamentHealingPerk,
                    RecyclePet_Selected_TextBlock_Current_TournamentRange,

                    RecyclePet_Selected_TextBlock_Current_PlayerPickEfficiency,
                    RecyclePet_Selected_TextBlock_Current_PlayerPickCriticalDamage,

                    AllRecyclePet_TextureBlock_PetImages,
                    AllRecyclePet_TextBlock_Classes,
                    AllRecyclePet_TextBlock_PackA,
                    AllRecyclePet_TextBlock_PackB,
                    AllRecyclePet_TextBlock_Purity,
                    AllRecyclePet_TextBlock_Tier
                    )

            if:
                PetToRecycle:=ValidBakery.PlayerDataCache.AllPets[RecyclePet_CurrentOption]
            then:
                TupleRecycleAmount:=CalculePetTotalAmountsInvested(PetToRecycle)
                RecyclePet_ExecuteRecycleButton.SetText(BasicMessage("RECYCLE"))
                RecyclePet_Text_MuntantMaterialRecicled.SetText(BasicMessage(GetMagnitudeLetter(TupleRecycleAmount(0)*RecycleEfficiency,"","")))
                RecyclePet_Text_CoalRecicled.SetText(BasicMessage(GetMagnitudeLetter((TupleRecycleAmount(1)+TupleRecycleAmount(2))*RecycleEfficiency,"","")))
                set IsRecycleReady = true
    MakeRecyclePet_Canvas(): canvas=
        Canvas := canvas:
            Slots:=array:
                #Main overlay
                canvas_slot:
                    Anchors:=anchors{Minimum:=vector2{X:=0.5,Y:=0.5},Maximum:=vector2{X:=0.5,Y:=0.5}}
                    ZOrder:=0
                    SizeToContent:=true
                    Offsets:=margin{Left:=0.0,Top:=0.0,Right:=0.0,Bottom:=0.0} #PositionX, PositionY, SizeX, SizeY
                    Alignment:=vector2{X:=0.5,Y:=0.5}
                    Widget:=texture_block{DefaultDesiredSize:=vector2{X:=3000.0,Y:=1250.0},DefaultImage:=Images.Img_MainBackground}                    
                canvas_slot:
                    Anchors:=anchors{Minimum:=vector2{X:=0.5,Y:=0.5},Maximum:=vector2{X:=0.5,Y:=0.5}}
                    ZOrder:=0
                    SizeToContent:=true
                    Offsets:=margin{Left:=0.0,Top:=-45.0,Right:=0.0,Bottom:=10.0} #PositionX, PositionY, SizeX, SizeY
                    Alignment:=vector2{X:=0.5,Y:=0.5}
                    Widget:=overlay:
                        Slots:=array:
                            overlay_slot:
                                HorizontalAlignment:=horizontal_alignment.Center
                                VerticalAlignment:=vertical_alignment.Center
                                Widget:=stack_box:  
                                    Orientation:=orientation.Horizontal
                                        Slots:=array:
                                            stack_box_slot: #Slot 0
                                                Padding:=margin{Right:=SuperLowPadding,Left:=SuperLowPadding,Bottom:=150.0}
                                                HorizontalAlignment:=horizontal_alignment.Center
                                                VerticalAlignment:=vertical_alignment.Bottom
                                                Widget:=stack_box:  
                                                    Orientation:=orientation.Vertical
                                                        Slots:=array:
                                                            stack_box_slot: 
                                                                Padding:=margin{Top:=SelectorBoxPadding,Bottom:=SelectorBoxPadding}
                                                                Widget:=overlay:
                                                                    Slots:=array:
                                                                        overlay_slot:
                                                                            HorizontalAlignment:=horizontal_alignment.Fill
                                                                            VerticalAlignment:=vertical_alignment.Fill
                                                                            Widget:=texture_block{DefaultImage:=Images.Img_LineBackground}
                                                                        overlay_slot:
                                                                            HorizontalAlignment:=horizontal_alignment.Fill
                                                                            VerticalAlignment:=vertical_alignment.Fill
                                                                            Padding:=margin{Left:=SelectorBoxPadding,Right:=SelectorBoxPadding,Top:=SelectorBoxPadding,Bottom:=SelectorBoxPadding}
                                                                            Widget:=texture_block{DefaultImage:=Images.Img_LineBackground}
                                                                        overlay_slot:
                                                                            HorizontalAlignment:=horizontal_alignment.Center
                                                                            VerticalAlignment:=vertical_alignment.Center
                                                                            Widget:=texture_block{DefaultImage:=Images.EmptyTexture,DefaultDesiredSize:=vector2{X:=258.0,Y:=300.0}}
                                                                        overlay_slot:
                                                                            HorizontalAlignment:=horizontal_alignment.Center
                                                                            VerticalAlignment:=vertical_alignment.Center
                                                                            Widget:=stack_box:
                                                                                Orientation:=orientation.Vertical
                                                                                Slots:=array:
                                                                                    stack_box_slot: 
                                                                                        Padding:=RowPadding
                                                                                        HorizontalAlignment:=horizontal_alignment.Center
                                                                                        VerticalAlignment:=vertical_alignment.Center
                                                                                        Widget:=RecyclePet_TextureBlock_PetImages_Box0
                                                                                    stack_box_slot: 
                                                                                        HorizontalAlignment:=horizontal_alignment.Center
                                                                                        VerticalAlignment:=vertical_alignment.Center
                                                                                        #Padding:=margin{Left:=TextPetsBoxPadding}
                                                                                        Widget:=stack_box:
                                                                                            Orientation:=orientation.Horizontal
                                                                                            Slots:=array:
                                                                                                stack_box_slot: 
                                                                                                    Padding:=margin{Left:=2.5,Right:=2.5}
                                                                                                    Widget:=RecyclePet_TextBlock_Classes_Box0
                                                                                                stack_box_slot: 
                                                                                                    Padding:=margin{Left:=2.5, Right:=1.0}
                                                                                                    Widget:=RecyclePet_TextBlock_PackA_Box0
                                                                                                stack_box_slot: 
                                                                                                    Padding:=margin{Right:=2.5}
                                                                                                    Widget:=RecyclePet_TextBlock_PackB_Box0
                                                                                    stack_box_slot: 
                                                                                        Padding:=RowPadding
                                                                                        HorizontalAlignment:=horizontal_alignment.Center
                                                                                        VerticalAlignment:=vertical_alignment.Center
                                                                                        Widget:=RecyclePet_TextBlock_Purity_Box0
                                                                                    stack_box_slot: 
                                                                                        Padding:=RowPadding
                                                                                        HorizontalAlignment:=horizontal_alignment.Center
                                                                                        VerticalAlignment:=vertical_alignment.Center
                                                                                        Widget:=RecyclePet_TextBlock_Tier_Box0                                    
                                                            stack_box_slot: 
                                                                Widget:=overlay:
                                                                    Slots:=array:
                                                                        overlay_slot:
                                                                            HorizontalAlignment:=horizontal_alignment.Fill
                                                                            VerticalAlignment:=vertical_alignment.Fill
                                                                            Widget:=texture_block{DefaultImage:=Images.EmptyTexture,DefaultDesiredSize:=QuietButton_PetSelectorDesiredSize}
                                            stack_box_slot: #Slot 1
                                                Padding:=margin{Right:=SuperLowPadding,Left:=SuperLowPadding,Bottom:=150.0}
                                                HorizontalAlignment:=horizontal_alignment.Center
                                                VerticalAlignment:=vertical_alignment.Bottom
                                                Widget:=stack_box:  
                                                    Orientation:=orientation.Vertical
                                                        Slots:=array:
                                                            stack_box_slot: 
                                                                Padding:=margin{Top:=SelectorBoxPadding,Bottom:=SelectorBoxPadding}
                                                                Widget:=overlay:
                                                                    Slots:=array:
                                                                        overlay_slot:
                                                                            HorizontalAlignment:=horizontal_alignment.Fill
                                                                            VerticalAlignment:=vertical_alignment.Fill
                                                                            Widget:=texture_block{DefaultImage:=Images.Img_LineBackground}
                                                                        overlay_slot:
                                                                            HorizontalAlignment:=horizontal_alignment.Fill
                                                                            VerticalAlignment:=vertical_alignment.Fill
                                                                            Padding:=margin{Left:=SelectorBoxPadding,Right:=SelectorBoxPadding,Top:=SelectorBoxPadding,Bottom:=SelectorBoxPadding}
                                                                            Widget:=texture_block{DefaultImage:=Images.Img_LineBackground}
                                                                        overlay_slot:
                                                                            HorizontalAlignment:=horizontal_alignment.Center
                                                                            VerticalAlignment:=vertical_alignment.Center
                                                                            Widget:=texture_block{DefaultImage:=Images.EmptyTexture,DefaultDesiredSize:=vector2{X:=258.0,Y:=300.0}}
                                                                        overlay_slot:
                                                                            HorizontalAlignment:=horizontal_alignment.Center
                                                                            VerticalAlignment:=vertical_alignment.Center
                                                                            Widget:=stack_box:
                                                                                Orientation:=orientation.Vertical
                                                                                Slots:=array:
                                                                                    stack_box_slot: 
                                                                                        Padding:=RowPadding
                                                                                        HorizontalAlignment:=horizontal_alignment.Center
                                                                                        VerticalAlignment:=vertical_alignment.Center
                                                                                        Widget:=RecyclePet_TextureBlock_PetImages_Box1
                                                                                    stack_box_slot: 
                                                                                        HorizontalAlignment:=horizontal_alignment.Center
                                                                                        VerticalAlignment:=vertical_alignment.Center
                                                                                        #Padding:=margin{Left:=TextPetsBoxPadding}
                                                                                        Widget:=stack_box:
                                                                                            Orientation:=orientation.Horizontal
                                                                                            Slots:=array:
                                                                                                stack_box_slot: 
                                                                                                    Padding:=margin{Left:=2.5,Right:=2.5}
                                                                                                    Widget:=RecyclePet_TextBlock_Classes_Box1
                                                                                                stack_box_slot: 
                                                                                                    Padding:=margin{Left:=2.5, Right:=1.0}
                                                                                                    Widget:=RecyclePet_TextBlock_PackA_Box1
                                                                                                stack_box_slot: 
                                                                                                    Padding:=margin{Right:=2.5}
                                                                                                    Widget:=RecyclePet_TextBlock_PackB_Box1
                                                                                    stack_box_slot: 
                                                                                        Padding:=RowPadding
                                                                                        HorizontalAlignment:=horizontal_alignment.Center
                                                                                        VerticalAlignment:=vertical_alignment.Center
                                                                                        Widget:=RecyclePet_TextBlock_Purity_Box1
                                                                                    stack_box_slot: 
                                                                                        Padding:=RowPadding
                                                                                        HorizontalAlignment:=horizontal_alignment.Center
                                                                                        VerticalAlignment:=vertical_alignment.Center
                                                                                        Widget:=RecyclePet_TextBlock_Tier_Box1
                                                            stack_box_slot: 
                                                                Widget:=overlay:
                                                                    Slots:=array:
                                                                        overlay_slot:
                                                                            HorizontalAlignment:=horizontal_alignment.Fill
                                                                            VerticalAlignment:=vertical_alignment.Fill
                                                                            Widget:=texture_block{DefaultImage:=Images.EmptyTexture,DefaultDesiredSize:=QuietButton_PetSelectorDesiredSize}

                                            stack_box_slot: #SELECTED
                                                Padding:=margin{Right:=SuperLowPadding,Left:=SuperLowPadding}
                                                HorizontalAlignment:=horizontal_alignment.Center
                                                VerticalAlignment:=vertical_alignment.Bottom
                                                Widget:=stack_box:  
                                                    Orientation:=orientation.Vertical
                                                    Slots:=array:
                                                        stack_box_slot: 
                                                            Padding:=margin{Top:=SelectorBoxPadding,Bottom:=SelectorBoxPadding}
                                                            Widget:=overlay:
                                                                Slots:=array:
                                                                    overlay_slot:
                                                                        HorizontalAlignment:=horizontal_alignment.Fill
                                                                        VerticalAlignment:=vertical_alignment.Fill
                                                                        Widget:=texture_block{DefaultImage:=Images.Img_LineBackground}
                                                                    overlay_slot:
                                                                        HorizontalAlignment:=horizontal_alignment.Fill
                                                                        VerticalAlignment:=vertical_alignment.Fill
                                                                        Padding:=margin{Left:=SelectorBoxPadding,Right:=SelectorBoxPadding,Top:=SelectorBoxPadding,Bottom:=SelectorBoxPadding}
                                                                        Widget:=texture_block{DefaultImage:=Images.Img_LineBackground}
                                                                    overlay_slot:
                                                                        HorizontalAlignment:=horizontal_alignment.Center
                                                                        VerticalAlignment:=vertical_alignment.Center
                                                                        Widget:=texture_block{DefaultImage:=Images.EmptyTexture,DefaultDesiredSize:=vector2{X:=1100.0,Y:=705.0}}
                                                                    overlay_slot:
                                                                        HorizontalAlignment:=horizontal_alignment.Center
                                                                        VerticalAlignment:=vertical_alignment.Center
                                                                        Widget:=stack_box:  
                                                                            Orientation:=orientation.Vertical
                                                                            Slots:=array:
                                                                                stack_box_slot: 
                                                                                    Widget:=stack_box:  
                                                                                        Orientation:=orientation.Horizontal
                                                                                        Slots:=array:
                                                                                            stack_box_slot: 
                                                                                                Padding:=RowPadding
                                                                                                Widget:=RecyclePet_Selected_TextureBlock_PetImages
                                                                                            stack_box_slot: 
                                                                                                HorizontalAlignment:=horizontal_alignment.Center
                                                                                                VerticalAlignment:=vertical_alignment.Center
                                                                                                Widget:=stack_box:  
                                                                                                    Orientation:=orientation.Vertical
                                                                                                    Slots:=array:
                                                                                                        stack_box_slot: 
                                                                                                            HorizontalAlignment:=horizontal_alignment.Center
                                                                                                            VerticalAlignment:=vertical_alignment.Center
                                                                                                            #Padding:=margin{Left:=TextPetsBoxPadding}
                                                                                                            Widget:=stack_box:
                                                                                                                Orientation:=orientation.Horizontal
                                                                                                                Slots:=array:
                                                                                                                    stack_box_slot: 
                                                                                                                        Padding:=margin{Left:=2.5,Right:=2.5}
                                                                                                                        Widget:=RecyclePet_Selected_TextBlock_Class
                                                                                                                    stack_box_slot: 
                                                                                                                        Padding:=margin{Left:=2.5, Right:=1.0}
                                                                                                                        Widget:=RecyclePet_Selected_TextBlock_PackA
                                                                                                                    stack_box_slot: 
                                                                                                                        Padding:=margin{Right:=2.5}
                                                                                                                        Widget:=RecyclePet_Selected_TextBlock_PackB
                                                                                                        stack_box_slot: 
                                                                                                            Padding:=RowPadding
                                                                                                            Widget:=RecyclePet_Selected_TextBlock_Purity
                                                                                stack_box_slot: 
                                                                                    HorizontalAlignment:=horizontal_alignment.Fill
                                                                                    VerticalAlignment:=vertical_alignment.Fill
                                                                                    Widget:=texture_block{DefaultImage:=Images.Img_HorizontalLine,DefaultDesiredSize:=vector2{Y:=3.0}}
                                                                                stack_box_slot:  
                                                                                    HorizontalAlignment:=horizontal_alignment.Center
                                                                                    VerticalAlignment:=vertical_alignment.Center
                                                                                    Padding:=RowPadding
                                                                                    Widget:=RecyclePet_Text_PackageA_LV
                                                                                stack_box_slot: 
                                                                                    HorizontalAlignment:=horizontal_alignment.Fill
                                                                                    VerticalAlignment:=vertical_alignment.Fill
                                                                                    Widget:=texture_block{DefaultImage:=Images.Img_HorizontalLine,DefaultDesiredSize:=vector2{Y:=3.0}}
                                                                                stack_box_slot: 
                                                                                    HorizontalAlignment:=horizontal_alignment.Fill
                                                                                    VerticalAlignment:=vertical_alignment.Fill
                                                                                    Widget:=stack_box:  
                                                                                        Orientation:=orientation.Horizontal
                                                                                        Slots:=array:
                                                                                            stack_box_slot: 
                                                                                                Widget:=stack_box:
                                                                                                    Orientation:=orientation.Vertical
                                                                                                    Slots:=array:
                                                                                                        stack_box_slot: 
                                                                                                            Padding:=RowPadding
                                                                                                            Widget:=text_block{DefaultText:=BasicMessage("Tournament stats"), DefaultTextColor:=Red, DefaultShadowOffset:=ShadowOffset, DefaultShadowOpacity := DefaultTextShadowOpacity}
                                                                                                        stack_box_slot: 
                                                                                                            HorizontalAlignment:=horizontal_alignment.Fill
                                                                                                            VerticalAlignment:=vertical_alignment.Fill
                                                                                                            Widget:=stack_box:  
                                                                                                                Orientation:=orientation.Horizontal
                                                                                                                Slots:=array:
                                                                                                                    stack_box_slot: 
                                                                                                                        HorizontalAlignment:=horizontal_alignment.Fill
                                                                                                                        VerticalAlignment:=vertical_alignment.Fill
                                                                                                                        Widget:=stack_box:  
                                                                                                                            Orientation:=orientation.Vertical
                                                                                                                            Slots:=array:
                                                                                                                                stack_box_slot: 
                                                                                                                                    Widget:=overlay:
                                                                                                                                        Slots:=array:
                                                                                                                                            overlay_slot:
                                                                                                                                                HorizontalAlignment:=horizontal_alignment.Left
                                                                                                                                                Padding:=RowPadding
                                                                                                                                                Widget:=text_block{DefaultText:=BasicMessage("Health:"), DefaultTextColor:=White, DefaultShadowOffset:=ShadowOffset, DefaultShadowOpacity := DefaultTextShadowOpacity}
                                                                                                                                            overlay_slot:
                                                                                                                                                Widget:=texture_block{DefaultImage:=Images.EmptyTexture,DefaultDesiredSize:=vector2{X:=230.0,Y:=32.0}} 
                                                                                                                                stack_box_slot: 
                                                                                                                                    HorizontalAlignment:=horizontal_alignment.Left
                                                                                                                                    Padding:=RowPadding
                                                                                                                                    Widget:=text_block{DefaultText:=BasicMessage("Damage:"), DefaultTextColor:=White, DefaultShadowOffset:=ShadowOffset, DefaultShadowOpacity := DefaultTextShadowOpacity}
                                                                                                                                stack_box_slot: 
                                                                                                                                    HorizontalAlignment:=horizontal_alignment.Left
                                                                                                                                    Padding:=RowPadding
                                                                                                                                    Widget:=text_block{DefaultText:=BasicMessage("Mov. speed:"), DefaultTextColor:=White, DefaultShadowOffset:=ShadowOffset, DefaultShadowOpacity := DefaultTextShadowOpacity}
                                                                                                                    stack_box_slot: 
                                                                                                                        HorizontalAlignment:=horizontal_alignment.Fill
                                                                                                                        VerticalAlignment:=vertical_alignment.Fill
                                                                                                                        Widget:=stack_box:  
                                                                                                                            Orientation:=orientation.Vertical
                                                                                                                            Slots:=array:
                                                                                                                                stack_box_slot: 
                                                                                                                                    Widget:=overlay:
                                                                                                                                        Slots:=array:
                                                                                                                                            overlay_slot:
                                                                                                                                                Padding:=RowPadding
                                                                                                                                                Widget:=RecyclePet_Selected_TextBlock_Current_TournamentHealth
                                                                                                                                            overlay_slot:
                                                                                                                                                Widget:=texture_block{DefaultImage:=Images.EmptyTexture,DefaultDesiredSize:=vector2{X:=257.0,Y:=32.0}}
                                                                                                                                stack_box_slot: 
                                                                                                                                    Padding:=RowPadding
                                                                                                                                    Widget:=RecyclePet_Selected_TextBlock_Current_TournamentDamage
                                                                                                                                stack_box_slot: 
                                                                                                                                    Padding:=RowPadding
                                                                                                                                    Widget:=RecyclePet_Selected_TextBlock_Current_TournamentSpeed
                                                                                            stack_box_slot: 
                                                                                                HorizontalAlignment:=horizontal_alignment.Center
                                                                                                VerticalAlignment:=vertical_alignment.Fill
                                                                                                Widget:=texture_block{DefaultImage:=Images.Img_VerticalLine,DefaultDesiredSize:=vector2{X:=10.0}}
                                                                                            stack_box_slot: 
                                                                                                Widget:=stack_box:
                                                                                                    Orientation:=orientation.Vertical
                                                                                                    Slots:=array:
                                                                                                        stack_box_slot: 
                                                                                                            Padding:=RowPadding
                                                                                                            Widget:=text_block{DefaultText:=BasicMessage("Player enhancements"), DefaultTextColor:=Cyan, DefaultShadowOffset:=ShadowOffset, DefaultShadowOpacity := DefaultTextShadowOpacity}
                                                                                                        stack_box_slot: 
                                                                                                            HorizontalAlignment:=horizontal_alignment.Fill
                                                                                                            VerticalAlignment:=vertical_alignment.Fill
                                                                                                            Widget:=stack_box:  
                                                                                                                Orientation:=orientation.Horizontal
                                                                                                                Slots:=array:
                                                                                                                    stack_box_slot: 
                                                                                                                        HorizontalAlignment:=horizontal_alignment.Fill
                                                                                                                        VerticalAlignment:=vertical_alignment.Fill
                                                                                                                        Widget:=stack_box:  
                                                                                                                            Orientation:=orientation.Vertical
                                                                                                                            Slots:=array:
                                                                                                                                stack_box_slot: 
                                                                                                                                    Widget:=overlay:
                                                                                                                                        Slots:=array:
                                                                                                                                            overlay_slot:
                                                                                                                                                Padding:=RowPadding
                                                                                                                                                HorizontalAlignment:=horizontal_alignment.Left
                                                                                                                                                Widget:=RecyclePet_Selected_TextBlock_Description_SpecialPerkA
                                                                                                                                            overlay_slot:
                                                                                                                                                Widget:=texture_block{DefaultImage:=Images.EmptyTexture,DefaultDesiredSize:=vector2{X:=300.0,Y:=32.0}}
                                                                                                                                stack_box_slot: 
                                                                                                                                    Padding:=RowPadding
                                                                                                                                    HorizontalAlignment:=horizontal_alignment.Left
                                                                                                                                    Widget:=text_block{DefaultText:=BasicMessage("Pick damage:"), DefaultTextColor:=White, DefaultShadowOffset:=ShadowOffset, DefaultShadowOpacity := DefaultTextShadowOpacity}
                                                                                                                                stack_box_slot: 
                                                                                                                                    Padding:=RowPadding
                                                                                                                                    HorizontalAlignment:=horizontal_alignment.Left
                                                                                                                                    Widget:=text_block{DefaultText:=BasicMessage("Weak point mult:"), DefaultTextColor:=White, DefaultShadowOffset:=ShadowOffset, DefaultShadowOpacity := DefaultTextShadowOpacity}
                                                                                                                    stack_box_slot: 
                                                                                                                        HorizontalAlignment:=horizontal_alignment.Fill
                                                                                                                        VerticalAlignment:=vertical_alignment.Fill
                                                                                                                        Widget:=stack_box:  
                                                                                                                            Orientation:=orientation.Vertical
                                                                                                                            Slots:=array:
                                                                                                                                stack_box_slot: 
                                                                                                                                    Widget:=overlay:
                                                                                                                                        Slots:=array:
                                                                                                                                            overlay_slot:
                                                                                                                                                Padding:=RowPadding
                                                                                                                                                Widget:=RecyclePet_Selected_TextBlock_Current_SpecialPerkA
                                                                                                                                            overlay_slot:
                                                                                                                                                Widget:=texture_block{DefaultImage:=Images.EmptyTexture,DefaultDesiredSize:=vector2{X:=180.0,Y:=32.0}}
                                                                                                                                stack_box_slot: 
                                                                                                                                    Padding:=RowPadding
                                                                                                                                    Widget:=RecyclePet_Selected_TextBlock_Current_PlayerPickDamage
                                                                                                                                stack_box_slot:
                                                                                                                                    Padding:=RowPadding 
                                                                                                                                    Widget:=RecyclePet_Selected_TextBlock_Current_PlayerWeakPoint
                                                                                stack_box_slot: 
                                                                                    HorizontalAlignment:=horizontal_alignment.Fill
                                                                                    VerticalAlignment:=vertical_alignment.Fill
                                                                                    Widget:=texture_block{DefaultImage:=Images.Img_HorizontalLine,DefaultDesiredSize:=vector2{Y:=3.0}}
                                                                                stack_box_slot:  
                                                                                    HorizontalAlignment:=horizontal_alignment.Center
                                                                                    VerticalAlignment:=vertical_alignment.Center
                                                                                    Padding:=RowPadding
                                                                                    Widget:=RecyclePet_Text_PackageB_LV
                                                                                stack_box_slot: 
                                                                                    HorizontalAlignment:=horizontal_alignment.Fill
                                                                                    VerticalAlignment:=vertical_alignment.Fill
                                                                                    Widget:=texture_block{DefaultImage:=Images.Img_HorizontalLine,DefaultDesiredSize:=vector2{Y:=3.0}}
                                                                                stack_box_slot: 
                                                                                    HorizontalAlignment:=horizontal_alignment.Fill
                                                                                    VerticalAlignment:=vertical_alignment.Fill
                                                                                    Widget:=stack_box:  
                                                                                        Orientation:=orientation.Horizontal
                                                                                        Slots:=array:
                                                                                            stack_box_slot: 
                                                                                                Widget:=stack_box:
                                                                                                    Orientation:=orientation.Vertical
                                                                                                    Slots:=array:
                                                                                                        stack_box_slot: 
                                                                                                            Padding:=RowPadding
                                                                                                            Widget:=text_block{DefaultText:=BasicMessage("Tournament stats"), DefaultTextColor:=Red, DefaultShadowOffset:=ShadowOffset, DefaultShadowOpacity := DefaultTextShadowOpacity}
                                                                                                        stack_box_slot: 
                                                                                                            HorizontalAlignment:=horizontal_alignment.Fill
                                                                                                            VerticalAlignment:=vertical_alignment.Fill
                                                                                                            Widget:=stack_box:  
                                                                                                                Orientation:=orientation.Horizontal
                                                                                                                Slots:=array:
                                                                                                                    stack_box_slot: 
                                                                                                                        HorizontalAlignment:=horizontal_alignment.Fill
                                                                                                                        VerticalAlignment:=vertical_alignment.Fill
                                                                                                                        Widget:=stack_box:  
                                                                                                                            Orientation:=orientation.Vertical
                                                                                                                            Slots:=array:
                                                                                                                                stack_box_slot: 
                                                                                                                                    Widget:=overlay:
                                                                                                                                        Slots:=array:
                                                                                                                                            overlay_slot:
                                                                                                                                                Padding:=RowPadding
                                                                                                                                                HorizontalAlignment:=horizontal_alignment.Left
                                                                                                                                                Widget:=text_block{DefaultText:=BasicMessage("Attack time:"), DefaultTextColor:=White, DefaultShadowOffset:=ShadowOffset, DefaultShadowOpacity := DefaultTextShadowOpacity}
                                                                                                                                            overlay_slot:
                                                                                                                                                Widget:=texture_block{DefaultImage:=Images.EmptyTexture,DefaultDesiredSize:=vector2{X:=230.0,Y:=32.0}}
                                                                                                                                stack_box_slot: 
                                                                                                                                    Padding:=RowPadding
                                                                                                                                    HorizontalAlignment:=horizontal_alignment.Left
                                                                                                                                    Widget:=text_block{DefaultText:=BasicMessage("Range:"), DefaultTextColor:=White, DefaultShadowOffset:=ShadowOffset, DefaultShadowOpacity := DefaultTextShadowOpacity}
                                                                                                                                stack_box_slot: 
                                                                                                                                    Padding:=RowPadding
                                                                                                                                    HorizontalAlignment:=horizontal_alignment.Left
                                                                                                                                    Widget:=text_block{DefaultText:=BasicMessage("Healing perk:"), DefaultTextColor:=White, DefaultShadowOffset:=ShadowOffset, DefaultShadowOpacity := DefaultTextShadowOpacity}
                                                                                                                    stack_box_slot: 
                                                                                                                        HorizontalAlignment:=horizontal_alignment.Fill
                                                                                                                        VerticalAlignment:=vertical_alignment.Fill
                                                                                                                        Widget:=stack_box:  
                                                                                                                            Orientation:=orientation.Vertical
                                                                                                                            Slots:=array:
                                                                                                                                stack_box_slot: 
                                                                                                                                    Widget:=overlay:
                                                                                                                                        Slots:=array:
                                                                                                                                            overlay_slot:
                                                                                                                                                Padding:=RowPadding
                                                                                                                                                Widget:=RecyclePet_Selected_TextBlock_Current_TournamentActionSpeed
                                                                                                                                            overlay_slot:
                                                                                                                                                Widget:=texture_block{DefaultImage:=Images.EmptyTexture,DefaultDesiredSize:=vector2{X:=257.0,Y:=32.0}}
                                                                                                                                stack_box_slot: 
                                                                                                                                    Padding:=RowPadding
                                                                                                                                    Widget:=RecyclePet_Selected_TextBlock_Current_TournamentRange
                                                                                                                                stack_box_slot: 
                                                                                                                                    Padding:=RowPadding
                                                                                                                                    Widget:=RecyclePet_Selected_TextBlock_Current_TournamentHealingPerk
                                                                                            stack_box_slot: 
                                                                                                HorizontalAlignment:=horizontal_alignment.Center
                                                                                                VerticalAlignment:=vertical_alignment.Fill
                                                                                                Widget:=texture_block{DefaultImage:=Images.Img_VerticalLine,DefaultDesiredSize:=vector2{X:=10.0}}
                                                                                            stack_box_slot: 
                                                                                                Widget:=stack_box:
                                                                                                    Orientation:=orientation.Vertical
                                                                                                    Slots:=array:
                                                                                                        stack_box_slot: 
                                                                                                            Padding:=RowPadding
                                                                                                            Widget:=text_block{DefaultText:=BasicMessage("Player enhancements"), DefaultTextColor:=Cyan, DefaultShadowOffset:=ShadowOffset, DefaultShadowOpacity := DefaultTextShadowOpacity}
                                                                                                        stack_box_slot: 
                                                                                                            HorizontalAlignment:=horizontal_alignment.Fill
                                                                                                            VerticalAlignment:=vertical_alignment.Fill
                                                                                                            Widget:=stack_box:  
                                                                                                                Orientation:=orientation.Horizontal
                                                                                                                Slots:=array:
                                                                                                                    stack_box_slot: 
                                                                                                                        HorizontalAlignment:=horizontal_alignment.Fill
                                                                                                                        VerticalAlignment:=vertical_alignment.Fill
                                                                                                                        Widget:=stack_box:  
                                                                                                                            Orientation:=orientation.Vertical
                                                                                                                            Slots:=array:
                                                                                                                                stack_box_slot: 
                                                                                                                                    Widget:=overlay:
                                                                                                                                        Slots:=array:
                                                                                                                                            overlay_slot:
                                                                                                                                                Padding:=RowPadding
                                                                                                                                                HorizontalAlignment:=horizontal_alignment.Left
                                                                                                                                                Widget:=RecyclePet_Selected_TextBlock_Description_SpecialPerkB
                                                                                                                                            overlay_slot:
                                                                                                                                                Widget:=texture_block{DefaultImage:=Images.EmptyTexture,DefaultDesiredSize:=vector2{X:=300.0,Y:=32.0}}
                                                                                                                                stack_box_slot: 
                                                                                                                                    Padding:=RowPadding
                                                                                                                                    HorizontalAlignment:=horizontal_alignment.Left
                                                                                                                                    Widget:=text_block{DefaultText:=BasicMessage("Pick efficiency:"), DefaultTextColor:=White, DefaultShadowOffset:=ShadowOffset, DefaultShadowOpacity := DefaultTextShadowOpacity}
                                                                                                                                stack_box_slot: 
                                                                                                                                    Padding:=RowPadding
                                                                                                                                    HorizontalAlignment:=horizontal_alignment.Left
                                                                                                                                    Widget:=text_block{DefaultText:=BasicMessage("Pick critical dam:"), DefaultTextColor:=White, DefaultShadowOffset:=ShadowOffset, DefaultShadowOpacity := DefaultTextShadowOpacity}
                                                                                                                    stack_box_slot: 
                                                                                                                        HorizontalAlignment:=horizontal_alignment.Fill
                                                                                                                        VerticalAlignment:=vertical_alignment.Fill
                                                                                                                        Widget:=stack_box:  
                                                                                                                            Orientation:=orientation.Vertical
                                                                                                                            Slots:=array:
                                                                                                                                stack_box_slot: 
                                                                                                                                    Widget:=overlay:
                                                                                                                                        Slots:=array:
                                                                                                                                            overlay_slot:
                                                                                                                                                Padding:=RowPadding
                                                                                                                                                Widget:=RecyclePet_Selected_TextBlock_Current_SpecialPerkB
                                                                                                                                            overlay_slot:
                                                                                                                                                Widget:=texture_block{DefaultImage:=Images.EmptyTexture,DefaultDesiredSize:=vector2{X:=180.0,Y:=32.0}}
                                                                                                                                stack_box_slot: 
                                                                                                                                    Padding:=RowPadding
                                                                                                                                    Widget:=RecyclePet_Selected_TextBlock_Current_PlayerPickEfficiency
                                                                                                                                stack_box_slot: 
                                                                                                                                    Padding:=RowPadding
                                                                                                                                    Widget:=RecyclePet_Selected_TextBlock_Current_PlayerPickCriticalDamage
                                                        <#stack_box_slot: #Buttons Here 
                                                            Padding:=margin{Top:=SelectorBoxPadding,Bottom:=SelectorBoxPadding}
                                                            Widget:=overlay:
                                                                Slots:=array:
                                                                    overlay_slot:
                                                                        Widget:=stack_box:  
                                                                            Orientation:=orientation.Horizontal
                                                                                Slots:=array:
                                                                                    stack_box_slot: 
                                                                                        Padding:=RowPadding 
                                                                                        Widget:=overlay:
                                                                                            Slots:=array:
                                                                                                overlay_slot:
                                                                                                    Widget:=texture_block{DefaultImage:=Images.EmptyTexture,DefaultDesiredSize:=vector2{X:=530.0,Y:=32.0}}
                                                                                                overlay_slot:
                                                                                                    HorizontalAlignment:=horizontal_alignment.Fill
                                                                                                    Widget:=RecyclePet_NavigateLeftButton
                                                                                    stack_box_slot: 
                                                                                        Padding:=RowPadding 
                                                                                        Widget:=overlay:
                                                                                            Slots:=array:
                                                                                                overlay_slot:
                                                                                                    Widget:=texture_block{DefaultImage:=Images.EmptyTexture,DefaultDesiredSize:=vector2{X:=530.0,Y:=32.0}}
                                                                                                overlay_slot:
                                                                                                    HorizontalAlignment:=horizontal_alignment.Fill
                                                                                                    Widget:=RecyclePet_NavigateRightButton#>        
                                                        stack_box_slot: #Recicle execute button
                                                            Padding:=margin{Top:=SelectorBoxPadding,Bottom:=SelectorBoxPadding}
                                                            Widget:=overlay:
                                                                Slots:=array:
                                                                    overlay_slot:
                                                                        Widget:=texture_block{DefaultImage:=Images.EmptyTexture,DefaultDesiredSize:=vector2{X:=1000.0,Y:=80.0}}
                                                                    overlay_slot: 
                                                                        HorizontalAlignment:=horizontal_alignment.Fill
                                                                        VerticalAlignment:=vertical_alignment.Fill
                                                                        Widget:=RecyclePet_ExecuteRecycleButton
                                                        stack_box_slot: #Recicle info button
                                                            Padding:=margin{Top:=SelectorBoxPadding,Bottom:=SelectorBoxPadding}
                                                            Widget:=overlay:
                                                                Slots:=array:
                                                                    <#
                                                                    overlay_slot:
                                                                        HorizontalAlignment:=horizontal_alignment.Fill
                                                                        VerticalAlignment:=vertical_alignment.Fill
                                                                        Widget:=texture_block{DefaultImage:=Images.Img_LineBackground}
                                                                    overlay_slot:
                                                                        HorizontalAlignment:=horizontal_alignment.Fill
                                                                        VerticalAlignment:=vertical_alignment.Fill
                                                                        Padding:=margin{Left:=SelectorBoxPadding,Right:=SelectorBoxPadding,Top:=SelectorBoxPadding,Bottom:=SelectorBoxPadding}
                                                                        Widget:=texture_block{DefaultImage:=Images.Img_LineBackground}#>
                                                                    overlay_slot:
                                                                        Widget:=stack_box:  
                                                                            Orientation:=orientation.Horizontal
                                                                                Slots:=array:
                                                                                    stack_box_slot:
                                                                                        HorizontalAlignment:=horizontal_alignment.Right
                                                                                        VerticalAlignment:=vertical_alignment.Center
                                                                                        Padding:=margin{Left:=5.0,Right:=5.0,Top:=5.0,Bottom:=5.0}
                                                                                        Widget:=text_block{DefaultText:=BasicMessage("YOU WILL RECEIVE"), DefaultTextColor := White, DefaultShadowOffset:=ShadowOffset, DefaultShadowOpacity := DefaultTextShadowOpacity}
                                                                                    stack_box_slot:
                                                                                        HorizontalAlignment:=horizontal_alignment.Right
                                                                                        VerticalAlignment:=vertical_alignment.Center
                                                                                        Padding:=margin{Left:=5.0,Right:=5.0,Top:=5.0,Bottom:=5.0}
                                                                                        Widget:=overlay:
                                                                                            Slots:=array: 
                                                                                                overlay_slot:
                                                                                                    Widget:=texture_block{DefaultDesiredSize:=vector2{X:=SizeBoxSaphiresX,Y:=SizeBoxSaphiresY},DefaultImage:=Images.EmptyTexture} 
                                                                                                overlay_slot:
                                                                                                    HorizontalAlignment:=horizontal_alignment.Fill
                                                                                                    VerticalAlignment:=vertical_alignment.Fill
                                                                                                    Widget:=texture_block{DefaultImage:=Images.Img_LineBackground}  
                                                                                                overlay_slot:
                                                                                                    HorizontalAlignment:=horizontal_alignment.Left
                                                                                                    VerticalAlignment:=vertical_alignment.Center
                                                                                                    Padding:=margin{Left:=Icon_LeftMargin}
                                                                                                    Widget:=texture_block{DefaultDesiredSize:=vector2{X:=IconSmallSizeX,Y:=IconSmallSizeY},DefaultImage:=Images.Mutant_Material_Icon}
                                                                                                overlay_slot:
                                                                                                    HorizontalAlignment:=horizontal_alignment.Left
                                                                                                    VerticalAlignment:=vertical_alignment.Center
                                                                                                    Padding:=margin{Left:=TextIcon_LeftMargin}
                                                                                                    Widget:=RecyclePet_Text_MuntantMaterialRecicled 
                                                                                    stack_box_slot:
                                                                                        HorizontalAlignment:=horizontal_alignment.Right
                                                                                        VerticalAlignment:=vertical_alignment.Center
                                                                                        Padding:=margin{Left:=5.0,Right:=5.0,Top:=5.0,Bottom:=5.0}
                                                                                        Widget:=overlay:
                                                                                            Slots:=array: 
                                                                                                overlay_slot:
                                                                                                    Widget:=texture_block{DefaultDesiredSize:=vector2{X:=SizeBoxSaphiresX,Y:=SizeBoxSaphiresY},DefaultImage:=Images.EmptyTexture} 
                                                                                                overlay_slot:
                                                                                                    HorizontalAlignment:=horizontal_alignment.Fill
                                                                                                    VerticalAlignment:=vertical_alignment.Fill
                                                                                                    Widget:=texture_block{DefaultImage:=Images.Img_LineBackground}  
                                                                                                overlay_slot:
                                                                                                    HorizontalAlignment:=horizontal_alignment.Left
                                                                                                    VerticalAlignment:=vertical_alignment.Center
                                                                                                    Padding:=margin{Left:=Icon_LeftMargin}
                                                                                                    Widget:=texture_block{DefaultDesiredSize:=vector2{X:=IconSmallSizeX,Y:=IconSmallSizeY},DefaultImage:=Images.Coal_Icon}
                                                                                                overlay_slot:
                                                                                                    HorizontalAlignment:=horizontal_alignment.Left
                                                                                                    VerticalAlignment:=vertical_alignment.Center
                                                                                                    Padding:=margin{Left:=TextIcon_LeftMargin}
                                                                                                    Widget:=RecyclePet_Text_CoalRecicled 
                                                        stack_box_slot: #Navigation Buttons
                                                            Padding:=margin{Top:=SelectorBoxPadding,Bottom:=SelectorBoxPadding}
                                                            Widget:=overlay:
                                                                Slots:=array:
                                                                    overlay_slot:
                                                                        Widget:=stack_box:  
                                                                            Orientation:=orientation.Horizontal
                                                                                Slots:=array:
                                                                                    stack_box_slot: 
                                                                                        Padding:=RowPadding 
                                                                                        Widget:=overlay:
                                                                                            Slots:=array:
                                                                                                overlay_slot:
                                                                                                    Widget:=texture_block{DefaultImage:=Images.EmptyTexture,DefaultDesiredSize:=vector2{X:=400.0,Y:=80.0}}
                                                                                                overlay_slot: 
                                                                                                    HorizontalAlignment:=horizontal_alignment.Fill
                                                                                                    VerticalAlignment:=vertical_alignment.Fill
                                                                                                    Widget:=RecyclePet_NavigateLeftButton
                                                                                    stack_box_slot: 
                                                                                        Padding:=RowPadding 
                                                                                        Widget:=overlay:
                                                                                            Slots:=array:
                                                                                                overlay_slot:
                                                                                                    Widget:=texture_block{DefaultImage:=Images.EmptyTexture,DefaultDesiredSize:=vector2{X:=400.0,Y:=80.0}}
                                                                                                overlay_slot: 
                                                                                                    HorizontalAlignment:=horizontal_alignment.Fill
                                                                                                    VerticalAlignment:=vertical_alignment.Fill
                                                                                                    Widget:=RecyclePet_NavigateRightButton                              
                                            stack_box_slot: #Slot 2
                                                Padding:=margin{Right:=SuperLowPadding,Left:=SuperLowPadding,Bottom:=150.0}
                                                HorizontalAlignment:=horizontal_alignment.Center
                                                VerticalAlignment:=vertical_alignment.Bottom
                                                Widget:=stack_box:  
                                                    Orientation:=orientation.Vertical
                                                        Slots:=array:
                                                            stack_box_slot: 
                                                                Padding:=margin{Top:=SelectorBoxPadding,Bottom:=SelectorBoxPadding}
                                                                Widget:=overlay:
                                                                    Slots:=array:
                                                                        overlay_slot:
                                                                            HorizontalAlignment:=horizontal_alignment.Fill
                                                                            VerticalAlignment:=vertical_alignment.Fill
                                                                            Widget:=texture_block{DefaultImage:=Images.Img_LineBackground}
                                                                        overlay_slot:
                                                                            HorizontalAlignment:=horizontal_alignment.Fill
                                                                            VerticalAlignment:=vertical_alignment.Fill
                                                                            Padding:=margin{Left:=SelectorBoxPadding,Right:=SelectorBoxPadding,Top:=SelectorBoxPadding,Bottom:=SelectorBoxPadding}
                                                                            Widget:=texture_block{DefaultImage:=Images.Img_LineBackground}
                                                                        overlay_slot:
                                                                            HorizontalAlignment:=horizontal_alignment.Center
                                                                            VerticalAlignment:=vertical_alignment.Center
                                                                            Widget:=texture_block{DefaultImage:=Images.EmptyTexture,DefaultDesiredSize:=vector2{X:=258.0,Y:=300.0}}
                                                                        overlay_slot:
                                                                            HorizontalAlignment:=horizontal_alignment.Center
                                                                            VerticalAlignment:=vertical_alignment.Center
                                                                            Widget:=stack_box:
                                                                                Orientation:=orientation.Vertical
                                                                                Slots:=array:
                                                                                    stack_box_slot: 
                                                                                        Padding:=RowPadding
                                                                                        HorizontalAlignment:=horizontal_alignment.Center
                                                                                        VerticalAlignment:=vertical_alignment.Center
                                                                                        Widget:=RecyclePet_TextureBlock_PetImages_Box2
                                                                                    stack_box_slot: 
                                                                                        HorizontalAlignment:=horizontal_alignment.Center
                                                                                        VerticalAlignment:=vertical_alignment.Center
                                                                                        #Padding:=margin{Left:=TextPetsBoxPadding}
                                                                                        Widget:=stack_box:
                                                                                            Orientation:=orientation.Horizontal
                                                                                            Slots:=array:
                                                                                                stack_box_slot: 
                                                                                                    Padding:=margin{Left:=2.5,Right:=2.5}
                                                                                                    Widget:=RecyclePet_TextBlock_Classes_Box2
                                                                                                stack_box_slot: 
                                                                                                    Padding:=margin{Left:=2.5, Right:=1.0}
                                                                                                    Widget:=RecyclePet_TextBlock_PackA_Box2
                                                                                                stack_box_slot: 
                                                                                                    Padding:=margin{Right:=2.5}
                                                                                                    Widget:=RecyclePet_TextBlock_PackB_Box2
                                                                                    stack_box_slot: 
                                                                                        Padding:=RowPadding
                                                                                        HorizontalAlignment:=horizontal_alignment.Center
                                                                                        VerticalAlignment:=vertical_alignment.Center
                                                                                        Widget:=RecyclePet_TextBlock_Purity_Box2
                                                                                    stack_box_slot: 
                                                                                        Padding:=RowPadding
                                                                                        HorizontalAlignment:=horizontal_alignment.Center
                                                                                        VerticalAlignment:=vertical_alignment.Center
                                                                                        Widget:=RecyclePet_TextBlock_Tier_Box2
                                                            stack_box_slot: 
                                                                Widget:=overlay:
                                                                    Slots:=array:
                                                                        overlay_slot:
                                                                            HorizontalAlignment:=horizontal_alignment.Fill
                                                                            VerticalAlignment:=vertical_alignment.Fill
                                                                            Widget:=texture_block{DefaultImage:=Images.EmptyTexture,DefaultDesiredSize:=QuietButton_PetSelectorDesiredSize}

                                            stack_box_slot: #Slot 3
                                                Padding:=margin{Right:=SuperLowPadding,Left:=SuperLowPadding,Bottom:=150.0}
                                                HorizontalAlignment:=horizontal_alignment.Center
                                                VerticalAlignment:=vertical_alignment.Bottom
                                                Widget:=stack_box:  
                                                    Orientation:=orientation.Vertical
                                                        Slots:=array:
                                                            stack_box_slot: 
                                                                Padding:=margin{Top:=SelectorBoxPadding,Bottom:=SelectorBoxPadding}
                                                                Widget:=overlay:
                                                                    Slots:=array:
                                                                        overlay_slot:
                                                                            HorizontalAlignment:=horizontal_alignment.Fill
                                                                            VerticalAlignment:=vertical_alignment.Fill
                                                                            Widget:=texture_block{DefaultImage:=Images.Img_LineBackground}
                                                                        overlay_slot:
                                                                            HorizontalAlignment:=horizontal_alignment.Fill
                                                                            VerticalAlignment:=vertical_alignment.Fill
                                                                            Padding:=margin{Left:=SelectorBoxPadding,Right:=SelectorBoxPadding,Top:=SelectorBoxPadding,Bottom:=SelectorBoxPadding}
                                                                            Widget:=texture_block{DefaultImage:=Images.Img_LineBackground}
                                                                        overlay_slot:
                                                                            HorizontalAlignment:=horizontal_alignment.Center
                                                                            VerticalAlignment:=vertical_alignment.Center
                                                                            Widget:=texture_block{DefaultImage:=Images.EmptyTexture,DefaultDesiredSize:=vector2{X:=258.0,Y:=300.0}}
                                                                        overlay_slot:
                                                                            HorizontalAlignment:=horizontal_alignment.Center
                                                                            VerticalAlignment:=vertical_alignment.Center
                                                                            Widget:=stack_box:
                                                                                Orientation:=orientation.Vertical
                                                                                Slots:=array:
                                                                                    stack_box_slot: 
                                                                                        Padding:=RowPadding
                                                                                        HorizontalAlignment:=horizontal_alignment.Center
                                                                                        VerticalAlignment:=vertical_alignment.Center
                                                                                        Widget:=RecyclePet_TextureBlock_PetImages_Box3
                                                                                    stack_box_slot: 
                                                                                        HorizontalAlignment:=horizontal_alignment.Center
                                                                                        VerticalAlignment:=vertical_alignment.Center
                                                                                        #Padding:=margin{Left:=TextPetsBoxPadding}
                                                                                        Widget:=stack_box:
                                                                                            Orientation:=orientation.Horizontal
                                                                                            Slots:=array:
                                                                                                stack_box_slot: 
                                                                                                    Padding:=margin{Left:=2.5,Right:=2.5}
                                                                                                    Widget:=RecyclePet_TextBlock_Classes_Box3
                                                                                                stack_box_slot: 
                                                                                                    Padding:=margin{Left:=2.5, Right:=1.0}
                                                                                                    Widget:=RecyclePet_TextBlock_PackA_Box3
                                                                                                stack_box_slot: 
                                                                                                    Padding:=margin{Right:=2.5}
                                                                                                    Widget:=RecyclePet_TextBlock_PackB_Box3
                                                                                    stack_box_slot: 
                                                                                        Padding:=RowPadding
                                                                                        HorizontalAlignment:=horizontal_alignment.Center
                                                                                        VerticalAlignment:=vertical_alignment.Center
                                                                                        Widget:=RecyclePet_TextBlock_Purity_Box3
                                                                                    stack_box_slot: 
                                                                                        Padding:=RowPadding
                                                                                        HorizontalAlignment:=horizontal_alignment.Center
                                                                                        VerticalAlignment:=vertical_alignment.Center
                                                                                        Widget:=RecyclePet_TextBlock_Tier_Box3
                                                            stack_box_slot: 
                                                                Widget:=overlay:
                                                                    Slots:=array:
                                                                        overlay_slot:
                                                                            HorizontalAlignment:=horizontal_alignment.Fill
                                                                            VerticalAlignment:=vertical_alignment.Fill
                                                                            Widget:=texture_block{DefaultImage:=Images.EmptyTexture,DefaultDesiredSize:=QuietButton_PetSelectorDesiredSize}
                canvas_slot:
                    Anchors:=anchors{Minimum:=vector2{X:=0.5,Y:=1.0},Maximum:=vector2{X:=0.5,Y:=1.0}}
                    ZOrder:=0
                    SizeToContent:=true
                    Offsets:=margin{Left:=0.0,Top:=-27.0,Right:=0.0,Bottom:=0.0} #PositionX, PositionY, SizeX, SizeY
                    Alignment:=vector2{X:=0.5,Y:=0.5}
                    Widget:=overlay:
                        Slots:=array:
                            overlay_slot:
                                Padding:=margin{Bottom:=RowTopHugePadding}
                                Widget:=stack_box:  
                                    Orientation:=orientation.Vertical
                                        Slots:=array:                                    
                                            stack_box_slot: 
                                                    Padding:=margin{Bottom:=RowTopHugePadding}
                                                    Widget:=overlay:
                                                        Slots:=array:
                                                            overlay_slot:
                                                                Widget:=texture_block{DefaultImage:=Images.EmptyTexture,DefaultDesiredSize:=vector2{X:=BigCloseButtonSizeX,Y:=80.0}}
                                                            overlay_slot: 
                                                                HorizontalAlignment:=horizontal_alignment.Fill
                                                                VerticalAlignment:=vertical_alignment.Fill
                                                                Widget:=RecyclePet_CancelButton


    #MIX PETS
    ##################                            
    Handler_MainMenu_To_MixPetsMenu(Message : widget_message) : void =
        spawn:
            FromMainMenu_To_MixPetsMenu(Message)

    Handler_MixPetsMenu_To_MainMenu(Message : widget_message)<suspends> : void =
        spawn:
            FromMixPetsMenu_To_MainMenu(Message)

    FromMainMenu_To_MixPetsMenu(Message : widget_message)<suspends> : void =
        CreatePet_Remove_Canvas(Message)
        Sleep(0.1)
        MixPets_Display_Canvas()

    FromMixPetsMenu_To_MainMenu(Message : widget_message)<suspends> : void =
        MixPets_Remove_Canvas(Message)
        Sleep(0.1)
        CreatePet_Display_Canvas()

                                                                                            
    MakeMixPets_Canvas(): canvas=
        Canvas := canvas:
            Slots:=array:
                #Main overlay
                canvas_slot:
                    Anchors:=anchors{Minimum:=vector2{X:=0.5,Y:=0.5},Maximum:=vector2{X:=0.5,Y:=0.5}}
                    ZOrder:=0
                    SizeToContent:=true
                    Offsets:=margin{Left:=0.0,Top:=0.0,Right:=0.0,Bottom:=0.0} #PositionX, PositionY, SizeX, SizeY
                    Alignment:=vector2{X:=0.5,Y:=0.5}
                    Widget:=texture_block{DefaultDesiredSize:=vector2{X:=3000.0,Y:=1250.0},DefaultImage:=Images.Img_MainBackground}
                canvas_slot:
                    Anchors:=anchors{Minimum:=vector2{X:=0.5,Y:=0.5},Maximum:=vector2{X:=0.5,Y:=0.5}}
                    ZOrder:=0
                    SizeToContent:=true
                    Offsets:=margin{Left:=0.0,Top:=-50.0,Right:=100.0,Bottom:=0.0} #PositionX, PositionY, SizeX, SizeY
                    Alignment:=vector2{X:=0.5,Y:=0.5}
                    Widget:=overlay:
                        Slots:=array:
                            overlay_slot:
                                Widget:=stack_box:  
                                    Orientation:=orientation.Horizontal
                                    Slots:=array:
                                        stack_box_slot: #Item to Mix 1
                                            Padding:=margin{Right:=SelectorBoxPadding,Left:=SelectorBoxPadding, Bottom:= 100.0}
                                            HorizontalAlignment:=horizontal_alignment.Center
                                            VerticalAlignment:=vertical_alignment.Bottom
                                            Widget:=stack_box:  
                                                Orientation:=orientation.Vertical
                                                    Slots:=array:
                                                        stack_box_slot: 
                                                            Padding:=margin{Top:=SelectorBoxPadding,Bottom:=SelectorBoxPadding}
                                                            Widget:=overlay:
                                                                Slots:=array:
                                                                    overlay_slot:
                                                                        HorizontalAlignment:=horizontal_alignment.Fill
                                                                        VerticalAlignment:=vertical_alignment.Fill
                                                                        Widget:=texture_block{DefaultImage:=Images.Img_LineBackground}
                                                                    overlay_slot:
                                                                        HorizontalAlignment:=horizontal_alignment.Fill
                                                                        VerticalAlignment:=vertical_alignment.Fill
                                                                        Padding:=margin{Top:=2.5,Bottom:=2.5,Left:=2.5,Right:=2.5}
                                                                        Widget:=texture_block{DefaultImage:=Images.Img_LineBackground}
                                                                    overlay_slot:
                                                                        HorizontalAlignment:=horizontal_alignment.Center
                                                                        VerticalAlignment:=vertical_alignment.Center
                                                                        Widget:=texture_block{DefaultImage:=Images.EmptyTexture,DefaultDesiredSize:=vector2{X:=230.0,Y:=350.0}}
                                                                    overlay_slot:
                                                                        HorizontalAlignment:=horizontal_alignment.Center
                                                                        VerticalAlignment:=vertical_alignment.Center
                                                                        Widget:=stack_box:
                                                                            Orientation:=orientation.Vertical
                                                                            Slots:=array:
                                                                                stack_box_slot: 
                                                                                    Padding:=RowPadding
                                                                                    HorizontalAlignment:=horizontal_alignment.Center
                                                                                    VerticalAlignment:=vertical_alignment.Center
                                                                                    Widget:=text_block{DefaultText:=BasicMessage("SOURDOUGH 1"), DefaultTextColor:=White, DefaultShadowOffset:=ShadowOffset, DefaultShadowOpacity := DefaultTextShadowOpacity}
                                                                                stack_box_slot: 
                                                                                    Padding:=RowPadding
                                                                                    HorizontalAlignment:=horizontal_alignment.Center
                                                                                    VerticalAlignment:=vertical_alignment.Center
                                                                                    Widget:=MixPets_TextureBlock_PetImages_Box0
                                                                                stack_box_slot: 
                                                                                    HorizontalAlignment:=horizontal_alignment.Center
                                                                                    VerticalAlignment:=vertical_alignment.Center
                                                                                    #Padding:=margin{Left:=TextPetsBoxPadding}
                                                                                    Widget:=stack_box:
                                                                                        Orientation:=orientation.Horizontal
                                                                                        Slots:=array:
                                                                                            stack_box_slot: 
                                                                                                Padding:=margin{Left:=2.5,Right:=2.5}
                                                                                                Widget:=MixPets_TextBlock_Classes_Box0
                                                                                            stack_box_slot: 
                                                                                                Padding:=margin{Left:=2.5, Right:=1.0}
                                                                                                Widget:=MixPets_TextBlock_PackA_Box0
                                                                                            stack_box_slot: 
                                                                                                Padding:=margin{Right:=2.5}
                                                                                                Widget:=MixPets_TextBlock_PackB_Box0
                                                                                stack_box_slot: 
                                                                                    Padding:=RowPadding
                                                                                    HorizontalAlignment:=horizontal_alignment.Center
                                                                                    VerticalAlignment:=vertical_alignment.Center
                                                                                    Widget:=MixPets_TextBlock_Purity_Box0
                                                                                stack_box_slot: 
                                                                                    Padding:=RowPadding
                                                                                    HorizontalAlignment:=horizontal_alignment.Center
                                                                                    VerticalAlignment:=vertical_alignment.Center
                                                                                    Widget:=MixPets_TextBlock_Tier_Box0                                    
                                                        stack_box_slot: 
                                                            Widget:=overlay:
                                                                Slots:=array:
                                                                    overlay_slot:
                                                                        HorizontalAlignment:=horizontal_alignment.Fill
                                                                        VerticalAlignment:=vertical_alignment.Fill
                                                                        Widget:=texture_block{DefaultImage:=Images.EmptyTexture,DefaultDesiredSize:=vector2{X:=230.0,Y:=80.0}}
                                                                    overlay_slot:
                                                                        HorizontalAlignment:=horizontal_alignment.Fill
                                                                        VerticalAlignment:=vertical_alignment.Fill
                                                                        Widget:=MixPets_Button_Box0 
                                        stack_box_slot: 
                                            Padding:=margin{Left:=5.0,Bottom:=-100.0,Right:=5.0}
                                            Widget:=text_block{DefaultText:=BasicMessage("+"), DefaultTextColor:=White, DefaultShadowOffset:=ShadowOffset, DefaultShadowOpacity := DefaultTextShadowOpacity}

                                        stack_box_slot: #Item to Mix 2
                                            Padding:=margin{Right:=SelectorBoxPadding,Left:=SelectorBoxPadding, Bottom:= 100.0}
                                            HorizontalAlignment:=horizontal_alignment.Center
                                            VerticalAlignment:=vertical_alignment.Bottom
                                            Widget:=stack_box:  
                                                Orientation:=orientation.Vertical
                                                    Slots:=array:
                                                        stack_box_slot: 
                                                            Padding:=margin{Top:=SelectorBoxPadding,Bottom:=SelectorBoxPadding}
                                                            Widget:=overlay:
                                                                Slots:=array:
                                                                    overlay_slot:
                                                                        HorizontalAlignment:=horizontal_alignment.Fill
                                                                        VerticalAlignment:=vertical_alignment.Fill
                                                                        Widget:=texture_block{DefaultImage:=Images.Img_LineBackground}
                                                                    overlay_slot:
                                                                        HorizontalAlignment:=horizontal_alignment.Fill
                                                                        VerticalAlignment:=vertical_alignment.Fill
                                                                        Padding:=margin{Top:=2.5,Bottom:=2.5,Left:=2.5,Right:=2.5}
                                                                        Widget:=texture_block{DefaultImage:=Images.Img_LineBackground}
                                                                    overlay_slot:
                                                                        HorizontalAlignment:=horizontal_alignment.Center
                                                                        VerticalAlignment:=vertical_alignment.Center
                                                                        Widget:=texture_block{DefaultImage:=Images.EmptyTexture,DefaultDesiredSize:=vector2{X:=230.0,Y:=350.0}}
                                                                    overlay_slot:
                                                                        HorizontalAlignment:=horizontal_alignment.Center
                                                                        VerticalAlignment:=vertical_alignment.Center
                                                                        Widget:=stack_box:
                                                                            Orientation:=orientation.Vertical
                                                                            Slots:=array:
                                                                                stack_box_slot: 
                                                                                    Padding:=RowPadding
                                                                                    HorizontalAlignment:=horizontal_alignment.Center
                                                                                    VerticalAlignment:=vertical_alignment.Center
                                                                                    Widget:=text_block{DefaultText:=BasicMessage("SOURDOUGH 2"), DefaultTextColor:=White, DefaultShadowOffset:=ShadowOffset, DefaultShadowOpacity := DefaultTextShadowOpacity}
                                                                                stack_box_slot: 
                                                                                    Padding:=RowPadding
                                                                                    HorizontalAlignment:=horizontal_alignment.Center
                                                                                    VerticalAlignment:=vertical_alignment.Center
                                                                                    Widget:=MixPets_TextureBlock_PetImages_Box1
                                                                                stack_box_slot: 
                                                                                    HorizontalAlignment:=horizontal_alignment.Center
                                                                                    VerticalAlignment:=vertical_alignment.Center
                                                                                    #Padding:=margin{Left:=TextPetsBoxPadding}
                                                                                    Widget:=stack_box:
                                                                                        Orientation:=orientation.Horizontal
                                                                                        Slots:=array:
                                                                                            stack_box_slot: 
                                                                                                Padding:=margin{Left:=2.5,Right:=2.5}
                                                                                                Widget:=MixPets_TextBlock_Classes_Box1
                                                                                            stack_box_slot: 
                                                                                                Padding:=margin{Left:=2.5, Right:=1.0}
                                                                                                Widget:=MixPets_TextBlock_PackA_Box1
                                                                                            stack_box_slot: 
                                                                                                Padding:=margin{Right:=2.5}
                                                                                                Widget:=MixPets_TextBlock_PackB_Box1
                                                                                stack_box_slot: 
                                                                                    Padding:=RowPadding
                                                                                    HorizontalAlignment:=horizontal_alignment.Center
                                                                                    VerticalAlignment:=vertical_alignment.Center
                                                                                    Widget:=MixPets_TextBlock_Purity_Box1
                                                                                stack_box_slot: 
                                                                                    Padding:=RowPadding
                                                                                    HorizontalAlignment:=horizontal_alignment.Center
                                                                                    VerticalAlignment:=vertical_alignment.Center
                                                                                    Widget:=MixPets_TextBlock_Tier_Box1                                   
                                                        stack_box_slot: 
                                                            Widget:=overlay:
                                                                Slots:=array:
                                                                    overlay_slot:
                                                                        HorizontalAlignment:=horizontal_alignment.Fill
                                                                        VerticalAlignment:=vertical_alignment.Fill
                                                                        Widget:=texture_block{DefaultImage:=Images.EmptyTexture,DefaultDesiredSize:=vector2{X:=230.0,Y:=80.0}}
                                                                    overlay_slot:
                                                                        HorizontalAlignment:=horizontal_alignment.Fill
                                                                        VerticalAlignment:=vertical_alignment.Fill
                                                                        Widget:=MixPets_Button_Box1
                                        stack_box_slot: 
                                            Padding:=margin{Left:=5.0,Bottom:=-100.0,Right:=5.0}
                                            Widget:=text_block{DefaultText:=BasicMessage("="), DefaultTextColor:=White, DefaultShadowOffset:=ShadowOffset, DefaultShadowOpacity := DefaultTextShadowOpacity}

                                        stack_box_slot: #Slot MIXED
                                            Padding:=margin{Right:=SelectorBoxPadding,Left:=SelectorBoxPadding}
                                            HorizontalAlignment:=horizontal_alignment.Center
                                            VerticalAlignment:=vertical_alignment.Bottom
                                            Widget:=stack_box:  
                                                Orientation:=orientation.Vertical
                                                Slots:=array:
                                                    stack_box_slot: #SELECTED
                                                        Padding:=margin{Right:=SuperLowPadding,Left:=SuperLowPadding}
                                                        HorizontalAlignment:=horizontal_alignment.Center
                                                        VerticalAlignment:=vertical_alignment.Bottom
                                                        Widget:=stack_box:  
                                                            Orientation:=orientation.Vertical
                                                            Slots:=array:
                                                                stack_box_slot: 
                                                                    Widget:=MixPets_MixTitle
                                                                stack_box_slot: 
                                                                    Padding:=margin{Top:=SelectorBoxPadding,Bottom:=SelectorBoxPadding}
                                                                    Widget:=overlay:
                                                                        Slots:=array:
                                                                            overlay_slot:
                                                                                HorizontalAlignment:=horizontal_alignment.Fill
                                                                                VerticalAlignment:=vertical_alignment.Fill
                                                                                Widget:=texture_block{DefaultImage:=Images.Img_LineBackground}
                                                                            overlay_slot:
                                                                                HorizontalAlignment:=horizontal_alignment.Fill
                                                                                VerticalAlignment:=vertical_alignment.Fill
                                                                                Padding:=margin{Left:=SelectorBoxPadding,Right:=SelectorBoxPadding,Top:=SelectorBoxPadding,Bottom:=SelectorBoxPadding}
                                                                                Widget:=texture_block{DefaultImage:=Images.Img_LineBackground}
                                                                            overlay_slot:
                                                                                HorizontalAlignment:=horizontal_alignment.Center
                                                                                VerticalAlignment:=vertical_alignment.Center
                                                                                Widget:=texture_block{DefaultImage:=Images.EmptyTexture,DefaultDesiredSize:=vector2{X:=1035.0,Y:=600.0}}
                                                                            overlay_slot:
                                                                                HorizontalAlignment:=horizontal_alignment.Center
                                                                                VerticalAlignment:=vertical_alignment.Center
                                                                                Widget:=stack_box:  
                                                                                    Orientation:=orientation.Vertical
                                                                                    Slots:=array:
                                                                                        stack_box_slot: 
                                                                                            Widget:=stack_box:  
                                                                                                Orientation:=orientation.Horizontal
                                                                                                Slots:=array:
                                                                                                    stack_box_slot: 
                                                                                                        Padding:=RowPadding
                                                                                                        Widget:=MixPets_Selected_TextureBlock_PetImages
                                                                                                    stack_box_slot: 
                                                                                                        HorizontalAlignment:=horizontal_alignment.Center
                                                                                                        VerticalAlignment:=vertical_alignment.Center
                                                                                                        Widget:=stack_box:  
                                                                                                            Orientation:=orientation.Vertical
                                                                                                            Slots:=array:
                                                                                                                stack_box_slot: 
                                                                                                                    HorizontalAlignment:=horizontal_alignment.Center
                                                                                                                    VerticalAlignment:=vertical_alignment.Center
                                                                                                                    #Padding:=margin{Left:=TextPetsBoxPadding}
                                                                                                                    Widget:=stack_box:
                                                                                                                        Orientation:=orientation.Horizontal
                                                                                                                        Slots:=array:
                                                                                                                            stack_box_slot: 
                                                                                                                                Padding:=margin{Left:=2.5,Right:=2.5}
                                                                                                                                Widget:=MixPets_Selected_TextBlock_Class
                                                                                                                            stack_box_slot: 
                                                                                                                                Padding:=margin{Left:=2.5, Right:=1.0}
                                                                                                                                Widget:=MixPets_Selected_TextBlock_PackA
                                                                                                                            stack_box_slot: 
                                                                                                                                Padding:=margin{Right:=2.5}
                                                                                                                                Widget:=MixPets_Selected_TextBlock_PackB
                                                                                                                stack_box_slot: 
                                                                                                                    Padding:=RowPadding
                                                                                                                    Widget:=MixPets_Selected_TextBlock_Purity
                                                                                        stack_box_slot: 
                                                                                            HorizontalAlignment:=horizontal_alignment.Fill
                                                                                            VerticalAlignment:=vertical_alignment.Fill
                                                                                            Widget:=texture_block{DefaultImage:=Images.Img_HorizontalLine,DefaultDesiredSize:=vector2{Y:=3.0}}
                                                                                        stack_box_slot:  
                                                                                            HorizontalAlignment:=horizontal_alignment.Center
                                                                                            VerticalAlignment:=vertical_alignment.Center
                                                                                            Padding:=RowPadding
                                                                                            Widget:=MixPets_Text_PackageA_LV
                                                                                        stack_box_slot: 
                                                                                            HorizontalAlignment:=horizontal_alignment.Fill
                                                                                            VerticalAlignment:=vertical_alignment.Fill
                                                                                            Widget:=texture_block{DefaultImage:=Images.Img_HorizontalLine,DefaultDesiredSize:=vector2{Y:=3.0}}
                                                                                        stack_box_slot: 
                                                                                            HorizontalAlignment:=horizontal_alignment.Fill
                                                                                            VerticalAlignment:=vertical_alignment.Fill
                                                                                            Widget:=stack_box:  
                                                                                                Orientation:=orientation.Horizontal
                                                                                                Slots:=array:
                                                                                                    stack_box_slot: 
                                                                                                        Widget:=stack_box:
                                                                                                            Orientation:=orientation.Vertical
                                                                                                            Slots:=array:
                                                                                                                stack_box_slot: 
                                                                                                                    Padding:=RowPadding
                                                                                                                    Widget:=text_block{DefaultText:=BasicMessage("Tournament stats"), DefaultTextColor:=Red, DefaultShadowOffset:=ShadowOffset, DefaultShadowOpacity := DefaultTextShadowOpacity}
                                                                                                                stack_box_slot: 
                                                                                                                    HorizontalAlignment:=horizontal_alignment.Fill
                                                                                                                    VerticalAlignment:=vertical_alignment.Fill
                                                                                                                    Widget:=stack_box:  
                                                                                                                        Orientation:=orientation.Horizontal
                                                                                                                        Slots:=array:
                                                                                                                            stack_box_slot: 
                                                                                                                                HorizontalAlignment:=horizontal_alignment.Fill
                                                                                                                                VerticalAlignment:=vertical_alignment.Fill
                                                                                                                                Widget:=stack_box:  
                                                                                                                                    Orientation:=orientation.Vertical
                                                                                                                                    Slots:=array:
                                                                                                                                        stack_box_slot: 
                                                                                                                                            Widget:=overlay:
                                                                                                                                                Slots:=array:
                                                                                                                                                    overlay_slot:
                                                                                                                                                        HorizontalAlignment:=horizontal_alignment.Left
                                                                                                                                                        Padding:=RowPadding
                                                                                                                                                        Widget:=text_block{DefaultText:=BasicMessage("Health:"), DefaultTextColor:=White, DefaultShadowOffset:=ShadowOffset, DefaultShadowOpacity := DefaultTextShadowOpacity}
                                                                                                                                                    overlay_slot:
                                                                                                                                                        Widget:=texture_block{DefaultImage:=Images.EmptyTexture,DefaultDesiredSize:=vector2{X:=230.0,Y:=32.0}} 
                                                                                                                                        stack_box_slot: 
                                                                                                                                            HorizontalAlignment:=horizontal_alignment.Left
                                                                                                                                            Padding:=RowPadding
                                                                                                                                            Widget:=text_block{DefaultText:=BasicMessage("Damage:"), DefaultTextColor:=White, DefaultShadowOffset:=ShadowOffset, DefaultShadowOpacity := DefaultTextShadowOpacity}
                                                                                                                                        stack_box_slot: 
                                                                                                                                            HorizontalAlignment:=horizontal_alignment.Left
                                                                                                                                            Padding:=RowPadding
                                                                                                                                            Widget:=text_block{DefaultText:=BasicMessage("Mov. speed:"), DefaultTextColor:=White, DefaultShadowOffset:=ShadowOffset, DefaultShadowOpacity := DefaultTextShadowOpacity}
                                                                                                                            stack_box_slot: 
                                                                                                                                HorizontalAlignment:=horizontal_alignment.Fill
                                                                                                                                VerticalAlignment:=vertical_alignment.Fill
                                                                                                                                Widget:=stack_box:  
                                                                                                                                    Orientation:=orientation.Vertical
                                                                                                                                    Slots:=array:
                                                                                                                                        stack_box_slot: 
                                                                                                                                            Widget:=overlay:
                                                                                                                                                Slots:=array:
                                                                                                                                                    overlay_slot:
                                                                                                                                                        Padding:=RowPadding
                                                                                                                                                        Widget:=MixPets_Selected_TextBlock_Current_TournamentHealth
                                                                                                                                                    overlay_slot:
                                                                                                                                                        Widget:=texture_block{DefaultImage:=Images.EmptyTexture,DefaultDesiredSize:=vector2{X:=257.0,Y:=32.0}}
                                                                                                                                        stack_box_slot: 
                                                                                                                                            Padding:=RowPadding
                                                                                                                                            Widget:=MixPets_Selected_TextBlock_Current_TournamentDamage
                                                                                                                                        stack_box_slot: 
                                                                                                                                            Padding:=RowPadding
                                                                                                                                            Widget:=MixPets_Selected_TextBlock_Current_TournamentSpeed
                                                                                                    stack_box_slot: 
                                                                                                        HorizontalAlignment:=horizontal_alignment.Center
                                                                                                        VerticalAlignment:=vertical_alignment.Fill
                                                                                                        Widget:=texture_block{DefaultImage:=Images.Img_VerticalLine,DefaultDesiredSize:=vector2{X:=10.0}}
                                                                                                    stack_box_slot: 
                                                                                                        Widget:=stack_box:
                                                                                                            Orientation:=orientation.Vertical
                                                                                                            Slots:=array:
                                                                                                                stack_box_slot: 
                                                                                                                    Padding:=RowPadding
                                                                                                                    Widget:=text_block{DefaultText:=BasicMessage("Player enhancements"), DefaultTextColor:=Cyan, DefaultShadowOffset:=ShadowOffset, DefaultShadowOpacity := DefaultTextShadowOpacity}
                                                                                                                stack_box_slot: 
                                                                                                                    HorizontalAlignment:=horizontal_alignment.Fill
                                                                                                                    VerticalAlignment:=vertical_alignment.Fill
                                                                                                                    Widget:=stack_box:  
                                                                                                                        Orientation:=orientation.Horizontal
                                                                                                                        Slots:=array:
                                                                                                                            stack_box_slot: 
                                                                                                                                HorizontalAlignment:=horizontal_alignment.Fill
                                                                                                                                VerticalAlignment:=vertical_alignment.Fill
                                                                                                                                Widget:=stack_box:  
                                                                                                                                    Orientation:=orientation.Vertical
                                                                                                                                    Slots:=array:
                                                                                                                                        stack_box_slot: 
                                                                                                                                            Widget:=overlay:
                                                                                                                                                Slots:=array:
                                                                                                                                                    overlay_slot:
                                                                                                                                                        Padding:=RowPadding
                                                                                                                                                        HorizontalAlignment:=horizontal_alignment.Left
                                                                                                                                                        Widget:=MixPets_Selected_TextBlock_Description_SpecialPerkA
                                                                                                                                                    overlay_slot:
                                                                                                                                                        Widget:=texture_block{DefaultImage:=Images.EmptyTexture,DefaultDesiredSize:=vector2{X:=300.0,Y:=32.0}}
                                                                                                                                        stack_box_slot: 
                                                                                                                                            Padding:=RowPadding
                                                                                                                                            HorizontalAlignment:=horizontal_alignment.Left
                                                                                                                                            Widget:=text_block{DefaultText:=BasicMessage("Pick damage:"), DefaultTextColor:=White, DefaultShadowOffset:=ShadowOffset, DefaultShadowOpacity := DefaultTextShadowOpacity}
                                                                                                                                        stack_box_slot: 
                                                                                                                                            Padding:=RowPadding
                                                                                                                                            HorizontalAlignment:=horizontal_alignment.Left
                                                                                                                                            Widget:=text_block{DefaultText:=BasicMessage("Weak point mult:"), DefaultTextColor:=White, DefaultShadowOffset:=ShadowOffset, DefaultShadowOpacity := DefaultTextShadowOpacity}
                                                                                                                            stack_box_slot: 
                                                                                                                                HorizontalAlignment:=horizontal_alignment.Fill
                                                                                                                                VerticalAlignment:=vertical_alignment.Fill
                                                                                                                                Widget:=stack_box:  
                                                                                                                                    Orientation:=orientation.Vertical
                                                                                                                                    Slots:=array:
                                                                                                                                        stack_box_slot: 
                                                                                                                                            Widget:=overlay:
                                                                                                                                                Slots:=array:
                                                                                                                                                    overlay_slot:
                                                                                                                                                        Padding:=RowPadding
                                                                                                                                                        Widget:=MixPets_Selected_TextBlock_Current_SpecialPerkA
                                                                                                                                                    overlay_slot:
                                                                                                                                                        Widget:=texture_block{DefaultImage:=Images.EmptyTexture,DefaultDesiredSize:=vector2{X:=180.0,Y:=32.0}}
                                                                                                                                        stack_box_slot: 
                                                                                                                                            Padding:=RowPadding
                                                                                                                                            Widget:=MixPets_Selected_TextBlock_Current_PlayerPickDamage
                                                                                                                                        stack_box_slot:
                                                                                                                                            Padding:=RowPadding 
                                                                                                                                            Widget:=MixPets_Selected_TextBlock_Current_PlayerWeakPoint
                                                                                        stack_box_slot: 
                                                                                            HorizontalAlignment:=horizontal_alignment.Fill
                                                                                            VerticalAlignment:=vertical_alignment.Fill
                                                                                            Widget:=texture_block{DefaultImage:=Images.Img_HorizontalLine,DefaultDesiredSize:=vector2{Y:=3.0}}
                                                                                        stack_box_slot:  
                                                                                            HorizontalAlignment:=horizontal_alignment.Center
                                                                                            VerticalAlignment:=vertical_alignment.Center
                                                                                            Padding:=RowPadding
                                                                                            Widget:=MixPets_Text_PackageB_LV
                                                                                        stack_box_slot: 
                                                                                            HorizontalAlignment:=horizontal_alignment.Fill
                                                                                            VerticalAlignment:=vertical_alignment.Fill
                                                                                            Widget:=texture_block{DefaultImage:=Images.Img_HorizontalLine,DefaultDesiredSize:=vector2{Y:=3.0}}
                                                                                        stack_box_slot: 
                                                                                            HorizontalAlignment:=horizontal_alignment.Fill
                                                                                            VerticalAlignment:=vertical_alignment.Fill
                                                                                            Widget:=stack_box:  
                                                                                                Orientation:=orientation.Horizontal
                                                                                                Slots:=array:
                                                                                                    stack_box_slot: 
                                                                                                        Widget:=stack_box:
                                                                                                            Orientation:=orientation.Vertical
                                                                                                            Slots:=array:
                                                                                                                stack_box_slot: 
                                                                                                                    Padding:=RowPadding
                                                                                                                    Widget:=text_block{DefaultText:=BasicMessage("Tournament stats"), DefaultTextColor:=Red, DefaultShadowOffset:=ShadowOffset, DefaultShadowOpacity := DefaultTextShadowOpacity}
                                                                                                                stack_box_slot: 
                                                                                                                    HorizontalAlignment:=horizontal_alignment.Fill
                                                                                                                    VerticalAlignment:=vertical_alignment.Fill
                                                                                                                    Widget:=stack_box:  
                                                                                                                        Orientation:=orientation.Horizontal
                                                                                                                        Slots:=array:
                                                                                                                            stack_box_slot: 
                                                                                                                                HorizontalAlignment:=horizontal_alignment.Fill
                                                                                                                                VerticalAlignment:=vertical_alignment.Fill
                                                                                                                                Widget:=stack_box:  
                                                                                                                                    Orientation:=orientation.Vertical
                                                                                                                                    Slots:=array:
                                                                                                                                        stack_box_slot: 
                                                                                                                                            Widget:=overlay:
                                                                                                                                                Slots:=array:
                                                                                                                                                    overlay_slot:
                                                                                                                                                        Padding:=RowPadding
                                                                                                                                                        HorizontalAlignment:=horizontal_alignment.Left
                                                                                                                                                        Widget:=text_block{DefaultText:=BasicMessage("Attack time:"), DefaultTextColor:=White, DefaultShadowOffset:=ShadowOffset, DefaultShadowOpacity := DefaultTextShadowOpacity}
                                                                                                                                                    overlay_slot:
                                                                                                                                                        Widget:=texture_block{DefaultImage:=Images.EmptyTexture,DefaultDesiredSize:=vector2{X:=230.0,Y:=32.0}}
                                                                                                                                        stack_box_slot: 
                                                                                                                                            Padding:=RowPadding
                                                                                                                                            HorizontalAlignment:=horizontal_alignment.Left
                                                                                                                                            Widget:=text_block{DefaultText:=BasicMessage("Range:"), DefaultTextColor:=White, DefaultShadowOffset:=ShadowOffset, DefaultShadowOpacity := DefaultTextShadowOpacity}
                                                                                                                                        stack_box_slot: 
                                                                                                                                            Padding:=RowPadding
                                                                                                                                            HorizontalAlignment:=horizontal_alignment.Left
                                                                                                                                            Widget:=text_block{DefaultText:=BasicMessage("Healing perk:"), DefaultTextColor:=White, DefaultShadowOffset:=ShadowOffset, DefaultShadowOpacity := DefaultTextShadowOpacity}
                                                                                                                            stack_box_slot: 
                                                                                                                                HorizontalAlignment:=horizontal_alignment.Fill
                                                                                                                                VerticalAlignment:=vertical_alignment.Fill
                                                                                                                                Widget:=stack_box:  
                                                                                                                                    Orientation:=orientation.Vertical
                                                                                                                                    Slots:=array:
                                                                                                                                        stack_box_slot: 
                                                                                                                                            Widget:=overlay:
                                                                                                                                                Slots:=array:
                                                                                                                                                    overlay_slot:
                                                                                                                                                        Padding:=RowPadding
                                                                                                                                                        Widget:=MixPets_Selected_TextBlock_Current_TournamentActionSpeed
                                                                                                                                                    overlay_slot:
                                                                                                                                                        Widget:=texture_block{DefaultImage:=Images.EmptyTexture,DefaultDesiredSize:=vector2{X:=257.0,Y:=32.0}}
                                                                                                                                        stack_box_slot: 
                                                                                                                                            Padding:=RowPadding
                                                                                                                                            Widget:=MixPets_Selected_TextBlock_Current_TournamentRange
                                                                                                                                        stack_box_slot: 
                                                                                                                                            Padding:=RowPadding
                                                                                                                                            Widget:=MixPets_Selected_TextBlock_Current_TournamentHealingPerk
                                                                                                    stack_box_slot: 
                                                                                                        HorizontalAlignment:=horizontal_alignment.Center
                                                                                                        VerticalAlignment:=vertical_alignment.Fill
                                                                                                        Widget:=texture_block{DefaultImage:=Images.Img_VerticalLine,DefaultDesiredSize:=vector2{X:=10.0}}
                                                                                                    stack_box_slot: 
                                                                                                        Widget:=stack_box:
                                                                                                            Orientation:=orientation.Vertical
                                                                                                            Slots:=array:
                                                                                                                stack_box_slot: 
                                                                                                                    Padding:=RowPadding
                                                                                                                    Widget:=text_block{DefaultText:=BasicMessage("Player enhancements"), DefaultTextColor:=Cyan, DefaultShadowOffset:=ShadowOffset, DefaultShadowOpacity := DefaultTextShadowOpacity}
                                                                                                                stack_box_slot: 
                                                                                                                    HorizontalAlignment:=horizontal_alignment.Fill
                                                                                                                    VerticalAlignment:=vertical_alignment.Fill
                                                                                                                    Widget:=stack_box:  
                                                                                                                        Orientation:=orientation.Horizontal
                                                                                                                        Slots:=array:
                                                                                                                            stack_box_slot: 
                                                                                                                                HorizontalAlignment:=horizontal_alignment.Fill
                                                                                                                                VerticalAlignment:=vertical_alignment.Fill
                                                                                                                                Widget:=stack_box:  
                                                                                                                                    Orientation:=orientation.Vertical
                                                                                                                                    Slots:=array:
                                                                                                                                        stack_box_slot: 
                                                                                                                                            Widget:=overlay:
                                                                                                                                                Slots:=array:
                                                                                                                                                    overlay_slot:
                                                                                                                                                        Padding:=RowPadding
                                                                                                                                                        HorizontalAlignment:=horizontal_alignment.Left
                                                                                                                                                        Widget:=MixPets_Selected_TextBlock_Description_SpecialPerkB
                                                                                                                                                    overlay_slot:
                                                                                                                                                        Widget:=texture_block{DefaultImage:=Images.EmptyTexture,DefaultDesiredSize:=vector2{X:=300.0,Y:=32.0}}
                                                                                                                                        stack_box_slot: 
                                                                                                                                            Padding:=RowPadding
                                                                                                                                            HorizontalAlignment:=horizontal_alignment.Left
                                                                                                                                            Widget:=text_block{DefaultText:=BasicMessage("Pick efficiency:"), DefaultTextColor:=White, DefaultShadowOffset:=ShadowOffset, DefaultShadowOpacity := DefaultTextShadowOpacity}
                                                                                                                                        stack_box_slot: 
                                                                                                                                            Padding:=RowPadding
                                                                                                                                            HorizontalAlignment:=horizontal_alignment.Left
                                                                                                                                            Widget:=text_block{DefaultText:=BasicMessage("Pick critical dam:"), DefaultTextColor:=White, DefaultShadowOffset:=ShadowOffset, DefaultShadowOpacity := DefaultTextShadowOpacity}
                                                                                                                            stack_box_slot: 
                                                                                                                                HorizontalAlignment:=horizontal_alignment.Fill
                                                                                                                                VerticalAlignment:=vertical_alignment.Fill
                                                                                                                                Widget:=stack_box:  
                                                                                                                                    Orientation:=orientation.Vertical
                                                                                                                                    Slots:=array:
                                                                                                                                        stack_box_slot: 
                                                                                                                                            Widget:=overlay:
                                                                                                                                                Slots:=array:
                                                                                                                                                    overlay_slot:
                                                                                                                                                        Padding:=RowPadding
                                                                                                                                                        Widget:=MixPets_Selected_TextBlock_Current_SpecialPerkB
                                                                                                                                                    overlay_slot:
                                                                                                                                                        Widget:=texture_block{DefaultImage:=Images.EmptyTexture,DefaultDesiredSize:=vector2{X:=180.0,Y:=32.0}}
                                                                                                                                        stack_box_slot: 
                                                                                                                                            Padding:=RowPadding
                                                                                                                                            Widget:=MixPets_Selected_TextBlock_Current_PlayerPickEfficiency
                                                                                                                                        stack_box_slot: 
                                                                                                                                            Padding:=RowPadding
                                                                                                                                            Widget:=MixPets_Selected_TextBlock_Current_PlayerPickCriticalDamage
                                                    stack_box_slot: #Buttons Here 
                                                        Padding:=margin{Top:=SelectorBoxPadding,Bottom:=SelectorBoxPadding}
                                                        Widget:=overlay:
                                                            Slots:=array:
                                                                overlay_slot:
                                                                    Widget:=stack_box:  
                                                                        Orientation:=orientation.Horizontal
                                                                            Slots:=array:
                                                                                stack_box_slot: 
                                                                                    Padding:=RowPadding 
                                                                                    Widget:=overlay:
                                                                                        Slots:=array:
                                                                                            overlay_slot:
                                                                                                Widget:=texture_block{DefaultImage:=Images.EmptyTexture,DefaultDesiredSize:=vector2{X:=960.0,Y:=185.0}}
                                                                                            overlay_slot:
                                                                                                HorizontalAlignment:=horizontal_alignment.Fill
                                                                                                VerticalAlignment:=vertical_alignment.Fill
                                                                                                Widget:=MixPets_MixButton

                canvas_slot:
                    Anchors:=anchors{Minimum:=vector2{X:=0.5,Y:=0.5},Maximum:=vector2{X:=0.5,Y:=0.5}}
                    ZOrder:=0
                    SizeToContent:=true
                    Offsets:=margin{Left:=0.0,Top:=480.0,Right:=0.0,Bottom:=0.0} #PositionX, PositionY, SizeX, SizeY
                    Alignment:=vector2{X:=0.5,Y:=0.5}
                    Widget:=overlay:
                        Slots:=array:
                            overlay_slot:
                                Widget:=texture_block{DefaultImage:=Images.EmptyTexture,DefaultDesiredSize:=vector2{X:=1300.0,Y:=80.0}}
                            overlay_slot:
                                HorizontalAlignment:=horizontal_alignment.Fill
                                VerticalAlignment:=vertical_alignment.Fill
                                Widget:=MixPets_CloseButton 
                              
                                                                                                
    MixPets_TryToMixPets(Message : widget_message) : void =
        if(IsFusionReady?):
            if:
                IndexToMix0:=AllMixPets_PetsToMix[0]
                IndexToMix1:=AllMixPets_PetsToMix[1]
                IndexToMix0<>IndexToMix1
            then: 
                if(ConfirmationToMix?):
                    spawn:
                        HandleMixPetsEffect()
                    if(Message.Player.IsActive[], ValidBakery := GameManager.GetAgentBakery_Transacted(agent[Message.Player])?):
                        var AllPlayerPets:[]pet_data_table=ValidBakery.PlayerDataCache.AllPets
                            if(IndexToMix0<IndexToMix1):
                                if:
                                    Agent:=agent[Message.Player]
                                    set AllPlayerPets[IndexToMix0] = FusionedPet
                                    NewPets := AllPlayerPets.RemoveElement[IndexToMix1]
                                    NewTable := GameManager.PlayersDataManager.SavePets(ValidBakery.PlayerDataCache,NewPets)
                                    #GameManager.PlayersDataManager.SaveNewData(Message.Player,PlayerData,NewTable)
                                    MaybeBakery:=GameManager.GetAgentBakery_Transacted(Agent)
                                then:
                                    ValidBakery.SetNewPlayerDataCache(NewTable)
                                    LoadEquippedPets(NewTable)
                            else:
                                if:
                                    Agent:=agent[Message.Player]
                                    set AllPlayerPets[IndexToMix1] = FusionedPet
                                    NewPets := AllPlayerPets.RemoveElement[IndexToMix0]
                                    NewTable := GameManager.PlayersDataManager.SavePets(ValidBakery.PlayerDataCache,NewPets)
                                    #GameManager.PlayersDataManager.SaveNewData(Message.Player,PlayerData,NewTable)
                                    MaybeBakery:=GameManager.GetAgentBakery_Transacted(Agent)
                                then:
                                    ValidBakery.SetNewPlayerDataCache(NewTable)
                                    spawn:
                                        ValidBakery.GameManager.CalculateAndSaveFame(Agent)
                                    LoadEquippedPets(NewTable)
                            MixPets_Selected_TextBlock_PackA.SetTextColor(White)
                            MixPets_Selected_TextBlock_PackB.SetTextColor(White)
                            MixPets_Selected_TextBlock_PackA.SetText(BasicMessage(""))
                            MixPets_Selected_TextBlock_PackB.SetText(BasicMessage(""))
                            MixPets_Selected_TextBlock_Description_SpecialPerkA.SetText(BasicMessage(""))
                            MixPets_Selected_TextBlock_Description_SpecialPerkB.SetText(BasicMessage(""))
                            MixPets_Selected_TextBlock_Description_SpecialPerkA.SetTextColor(White)
                            MixPets_Selected_TextBlock_Description_SpecialPerkB.SetTextColor(White)
                            set AllMixPets_PetsToMix = array{-1,-1}
                            set FusionedPet = pet_data_table{}
                            set ConfirmationToMix = false
                            set IsFusionReady = false
                else:
                    set ConfirmationToMix = true
                    MixPets_MixButton.SetText(BasicMessage("ARE YOU SURE?"))
            else:
                spawn:
                    ShowMixtureError("ERROR: DUPLICATED SOURDOUGH","MIX SOURDOUGHS")
        else:
            spawn:
                ShowMixtureError("NOT READY!","MIX SOURDOUGHS")

    HandleMixPetsEffect()<suspends> : void =
        var TotalProgress : int = 0
        loop:
            set TotalProgress = TotalProgress + 5
            MixPets_MixButton.SetText(BasicMessage("MIXING SOURDOUGHS {TotalProgress}%"))
            Sleep(0.04)   
            if(TotalProgress>=100):
                break
        MixPets_MixButton.SetText(BasicMessage("SUCCESS!"))
        MixPets_RefreshAll()
        MixPets_MixButton.SetText(BasicMessage("MIX SOURDOUGHS"))


    ShowMixtureError(ErrorString : string, PostString : string)<suspends> : void =
        MixPets_MixButton.SetText(BasicMessage(ErrorString))
        Sleep(1.0)
        MixPets_MixButton.SetText(BasicMessage("{PostString}"))

                                                                                
    MixPets_Remove_Canvas(Message : widget_message) : void =
        set MixPets_UIIsDisplayed = false
        if:
            Agent := MaybeAgent?
            Player:=player[Agent]
            PlayerUI:=GetPlayerUI[Player]
            Canvas := MixPets_MaybeCanvas?
        then:
            PlayerUI.RemoveWidget(Canvas)
            MixPets_RemoveDelay_OnRemove()

    MixPets_Remove_CanvasWithoutMessage() : void =
        set MixPets_UIIsDisplayed = false
        if:
            Agent := MaybeAgent?
            Player:=player[Agent]
            PlayerUI:=GetPlayerUI[Player]
            Canvas := MixPets_MaybeCanvas?
        then:
            PlayerUI.RemoveWidget(Canvas)
            MixPets_RemoveDelay_OnRemove()

    MixPets_RemoveDelay_OnAdd() : void =
        MixPets_CloseButton.SetText(BasicMessage("Close"))

    MixPets_RemoveDelay_OnRemove() : void =
        MixPets_CloseButton.SetText(BasicMessage("Close "))
    
    MixPets_Display_Canvas() : void =
        if(not MixPets_UIIsDisplayed?):
            set ConfirmationToMix = false
            if:
                Agent := MaybeAgent?
                PlayerUI:=GetPlayerUI[player[Agent]]
                Canvas := MixPets_MaybeCanvas?
            then:
                PlayerUI.AddWidget(Canvas,player_ui_slot{ZOrder:=0, InputMode := ui_input_mode.All})
                MixPets_RemoveDelay_OnAdd()
                spawn:
                    MixPets_RefreshAll()

    MixPets_RefreshAll()<suspends> : void =
        set IsFusionReady=false
        if(Agent := MaybeAgent?, Player := player[Agent], Player.IsActive[], ValidBakery := GameManager.GetAgentBakery_Transacted(Agent)?):
            AllPets:=ValidBakery.PlayerDataCache.AllPets
            #Cheking if the pets are the same
            if:
                IndexPetToMix0:=AllMixPets_PetsToMix[0]
                IndexPetToMix1:=AllMixPets_PetsToMix[1]
                IndexPetToMix0<>-1
                IndexPetToMix1<>-1
                IndexPetToMix0=IndexPetToMix1
                set AllMixPets_PetsToMix[0] = -1

            #Two pets to mix
            for(Index:=0..AllMixPets_PetsToMix.Length-1, PetToMixIndex:=AllMixPets_PetsToMix[Index]):
                if:
                    TextureBlock_PetImage:=AllMixPets_TextureBlock_PetImages[Index]
                    TextBlock_Class:=AllMixPets_TextBlock_Classes[Index]
                    TextBlock_PackA:=AllMixPets_TextBlock_PackA[Index]
                    TextBlock_PackB:=AllMixPets_TextBlock_PackB[Index]
                    TextBlock_Purity:=AllMixPets_TextBlock_Purity[Index]
                    TextBlock_Tier:=AllMixPets_TextBlock_Tier[Index]
                then: 
                    if(Validpet:=AllPets[PetToMixIndex]):
                        TuplePurity:=ConvertRawPurity(Validpet.Purity)
                        SetPetImage(Validpet.Type, TextureBlock_PetImage)
                        SetPetClass(Validpet, TextBlock_Class, TextBlock_PackA, TextBlock_PackB)
                        SetPetPurity(TuplePurity(1), TextBlock_Purity)
                        SetPetTier(TuplePurity(0), TextBlock_Tier)
                    else:
                        TextureBlock_PetImage.SetImage(Images.Cross_Icon)
                        TextBlock_Class.SetText(BasicMessage("-"))
                        TextBlock_PackA.SetText(BasicMessage("-"))
                        TextBlock_PackB.SetText(BasicMessage("-"))
                        TextBlock_PackA.SetTextColor(White)
                        TextBlock_PackB.SetTextColor(White)
                        TextBlock_Purity.SetText(BasicMessage("-"))
                        TextBlock_Tier.SetText(BasicMessage("-"))
                        TextureBlock_PetImage.SetTint(White)
                        TextBlock_Class.SetTextColor(White)
                        TextBlock_Purity.SetTextColor(White)
                        TextBlock_Tier.SetTextColor(White)

            #Calcule Mix
            MixPets_Selected_TextureBlock_PetImages.SetImage(Images.Cross_Icon)
            MixPets_Selected_TextBlock_Purity.SetText(BasicMessage("-"))
            MixPets_Selected_TextBlock_Class.SetText(BasicMessage("-"))
            MixPets_Selected_TextBlock_Current_TournamentHealth.SetText(BasicMessage("-"))
            MixPets_Selected_TextBlock_Current_TournamentDamage.SetText(BasicMessage("-"))
            MixPets_Selected_TextBlock_Current_TournamentSpeed.SetText(BasicMessage("-"))

            MixPets_Selected_TextBlock_Current_SpecialPerkA.SetText(BasicMessage("-"))
            MixPets_Selected_TextBlock_Current_PlayerPickDamage.SetText(BasicMessage("-"))
            MixPets_Selected_TextBlock_Current_PlayerWeakPoint.SetText(BasicMessage("-"))

            MixPets_Selected_TextBlock_Current_TournamentActionSpeed.SetText(BasicMessage("-"))
            MixPets_Selected_TextBlock_Current_TournamentHealingPerk.SetText(BasicMessage("-"))
            MixPets_Selected_TextBlock_Current_TournamentRange.SetText(BasicMessage("-"))

            MixPets_Selected_TextBlock_Current_SpecialPerkB.SetText(BasicMessage("-"))
            MixPets_Selected_TextBlock_Current_PlayerPickEfficiency.SetText(BasicMessage("-"))
            MixPets_Selected_TextBlock_Current_PlayerPickCriticalDamage.SetText(BasicMessage("-"))
            if:
                IndexPetToMix0:=AllMixPets_PetsToMix[0]
                IndexPetToMix1:=AllMixPets_PetsToMix[1]
                IndexPetToMix0<>-1
                IndexPetToMix1<>-1
                PetToFusion0:=AllPets[IndexPetToMix0]
                PetToFusion1:=AllPets[IndexPetToMix1]
            then:
                var SelectedEmoji_SpecialA : string = ""
                var SelectedEmoji_SpecialB : string = ""
                if(PetToFusion0.Type <> PetToFusion1.Type):
                    MixPets_MixButton.SetText(BasicMessage("ERROR: SOURDOUGH TYPES ARE DIFFERENT"))
                    for(Texture:AllMixPets_TextureBlock_PetImages):
                        Texture.SetTint(Red)
                else:
                    for(Texture:AllMixPets_TextureBlock_PetImages):
                        Texture.SetTint(White)
                    MixPets_MixButton.SetText(BasicMessage("Calculating mixture..."))
                    CalculatingMixtureTextureBlock.SetImage(Images.BlackBackground)
                    CalculatingMixtureTextureBlock.SetTint(Red)
                    TupleFusionedPet:=GameManager.PlayersDataManager.CalculePetFusion(PetToFusion0,PetToFusion1,ValidBakery.PlayerDataCache.PetTier)
                    set FusionedPet=TupleFusionedPet(0)
                    set FusionedPetTierWarning = TupleFusionedPet(1)
                    CalculatingMixtureTextureBlock.SetImage(Images.EmptyTexture)
                    CalculatingMixtureTextureBlock.SetTint(White)
                    Sleep(0.5)
                    set IsFusionReady=true
                    if(FusionedPetTierWarning?):
                        MixPets_MixTitle.SetText(BasicMessage("Purity limited! Recommended to unlock the next purity tier first"))
                        MixPets_MixTitle.SetTextColor(Red)
                        MixPets_MixButton.SetText(BasicMessage("Warning: purity limited. Unlock Purity tier"))
                    else:
                        MixPets_MixButton.SetText(BasicMessage("MIX SOURDOUGHS"))
                        MixPets_MixButton.SetText(BasicMessage("SOURDOUGH MIXTURE OUTPUT"))
                        MixPets_MixTitle.SetTextColor(Purple)
                    RefreshPetStats(
                        -1,
                        array{},
                        option{FusionedPet},
                        MixPets_Selected_TextureBlock_PetImages,

                        MixPets_Text_PackageA_LV,
                        MixPets_Text_PackageB_LV,
                        
                        MixPets_Selected_TextBlock_Current_SpecialPerkA,
                        MixPets_Selected_TextBlock_Current_SpecialPerkB,
                        MixPets_Selected_TextBlock_Description_SpecialPerkA,
                        MixPets_Selected_TextBlock_Description_SpecialPerkB,
                        MixPets_Selected_TextBlock_Purity,
                        MixPets_Selected_TextBlock_Class,
                        MixPets_Selected_TextBlock_PackA,
                        MixPets_Selected_TextBlock_PackB,
                        
                        MixPets_Selected_TextBlock_Current_TournamentHealth,
                        MixPets_Selected_TextBlock_Current_TournamentDamage,
                        MixPets_Selected_TextBlock_Current_TournamentSpeed,

                        MixPets_Selected_TextBlock_Current_PlayerPickDamage,
                        MixPets_Selected_TextBlock_Current_PlayerWeakPoint,

                        MixPets_Selected_TextBlock_Current_TournamentActionSpeed,
                        MixPets_Selected_TextBlock_Current_TournamentHealingPerk,
                        MixPets_Selected_TextBlock_Current_TournamentRange,

                        MixPets_Selected_TextBlock_Current_PlayerPickEfficiency,
                        MixPets_Selected_TextBlock_Current_PlayerPickCriticalDamage,

                        array{},
                        array{},
                        array{},
                        array{},
                        array{},
                        array{}
                        
                        )

    AnimatePetMovement(ObjetiveAgent : agent, Prop : creative_prop, Tier : int, PetSlot : int)<suspends> : void =
        if(OwnerCharacter := ObjetiveAgent.GetFortCharacter[]):

            var PivotVariation : vector3 = vector3{}
            var NormalScale : vector3 = vector3{}
            var FinalTransform : transform = transform{}

            OwnerTranslation:=OwnerCharacter.GetTransform().Translation
            PetTranslation:=Prop.GetTransform().Translation

            if(Tier=1):
                set NormalScale = PetScale_Tier1
                set PivotVariation = PetPivotVariation_Tier1
            if(Tier=2):
                set NormalScale = PetScale_Tier2
                set PivotVariation = PetPivotVariation_Tier2
            if(Tier=3):
                set NormalScale = PetScale_Tier3
                set PivotVariation = PetPivotVariation_Tier3

            
            ObjetiveRotation:=GetForwardRotation_Vector2(OwnerCharacter.GetTransform().Translation,Prop.GetTransform().Translation)
            set FinalTransform.Rotation = ObjetiveRotation
            set FinalTransform.Scale = NormalScale

            if(OwnerCharacter.IsCrouching[]):
                set PivotVariation = PivotVariation + PetPlayerCorrectionCrouching
            else:
                set PivotVariation = PivotVariation+  PetPlayerCorrection
                
            Vector:=OwnerCharacter.GetTransform().Translation+PivotVariation

            DistanceFromPlayer:=Distance(OwnerCharacter.GetTransform().Translation,Prop.GetTransform().Translation)

            if(DistanceFromPlayer<500.0, Prop.IsValid[]):
                if(DistanceFromPlayer<200.0):
                    DirectionVector:=vector3{X:=PetTranslation.X-OwnerTranslation.X,Y:=PetTranslation.Y-OwnerTranslation.Y}
                    set FinalTransform.Translation=PetTranslation+DirectionVector
                    Prop.MoveTo(FinalTransform,0.8)
                else:
                    FinalTranslation:=vector3{X:=Prop.GetTransform().Translation.X,Y:=Prop.GetTransform().Translation.Y,Z:=Vector.Z}
                    set FinalTransform.Translation = FinalTranslation
                    Prop.MoveTo(FinalTransform,0.8)
            else:
      
                ViewLocation:=OwnerCharacter.GetViewLocation()
                ViewRotation:=OwnerCharacter.GetViewRotation()

                var PointingVectorX: float =ViewRotation.GetLocalForward().X*(-15.0)
                var PointingVectorY: float =ViewRotation.GetLocalForward().Y*(-15.0)

                if(PetSlot=1):
                    set PointingVectorX = PointingVectorX + GetRandomFloat(34.0,48.0)
                    set PointingVectorY = PointingVectorY + GetRandomFloat(34.0,48.0)
                else if(PetSlot=2):
                    set PointingVectorX = PointingVectorX + GetRandomFloat(82.0,100.0)
                    set PointingVectorY = PointingVectorY + GetRandomFloat(82.0,100.0)

                NewPosition:=ViewLocation+vector3{X:=PointingVectorX,Y:=PointingVectorY}

                FinalTranslation:=vector3{X:=NewPosition.X,Y:=NewPosition.Y,Z:=Vector.Z}

                set FinalTransform=transform{Translation:=FinalTranslation,Rotation:=ObjetiveRotation,Scale:=NormalScale}
                if(Prop.IsValid[]):
                    Prop.MoveTo(FinalTransform,0.8)
            
           
MakePetsManagerUI<constructor>(InGameManager : game_manager)<transacts> := pets_manager :
    GameManager := InGameManager