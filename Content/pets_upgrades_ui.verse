# fill this file with your verse-script
using { /UnrealEngine.com/Temporary/UI }
using { /UnrealEngine.com/Temporary/SpatialMath }
using { /Fortnite.com/UI }
using { /Verse.org/Simulation }
using { /Verse.org/Colors/NamedColors }

# A Verse-authored creative device that can be placed in a level
pets_upgrades_ui := class():

    #Class settings
    var MaybeAgent : ?agent = false
    var PetsManager : pets_manager

    RowPadding:margin=margin{Left:=5.0,Top:=5.0,Right:=5.0,Bottom:=5.0}
    var MaybeCanvas : ?canvas = false
    var UIIsDisplayed : logic = false

    #Buttons
    Upgrade_AStatsButton : button_quiet = button_quiet{}
    Upgrade_BStatsButton : button_quiet = button_quiet{}

    NavigateRightButton : button_regular = button_regular{}
    NavigateLeftButton : button_regular = button_regular{}

    CancelButton : button_loud = button_loud{}

    var CurrentOption : int = 0

    TotalCoal_TextBlock : text_block = text_block{DefaultTextColor := CoalTextColor}

    #Box0
    TextureBlock_PetImages_Box0 : texture_block = texture_block{DefaultImage:=Images.EmptyTexture,DefaultDesiredSize:=PetsSelectorImageSize}
    TextBlock_Classes_Box0 : text_block = text_block{DefaultTextColor := White, DefaultShadowOffset:=ShadowOffset, DefaultShadowOpacity := DefaultTextShadowOpacity}
    TextBlock_PackA_Box0 : text_block = text_block{DefaultTextColor := White, DefaultShadowOffset:=ShadowOffset, DefaultShadowOpacity := DefaultTextShadowOpacity}
    TextBlock_PackB_Box0 : text_block = text_block{DefaultTextColor := White, DefaultShadowOffset:=ShadowOffset, DefaultShadowOpacity := DefaultTextShadowOpacity}
    TextBlock_Purity_Box0 : text_block = text_block{DefaultTextColor := White, DefaultShadowOffset:=ShadowOffset, DefaultShadowOpacity := DefaultTextShadowOpacity}
    TextBlock_Tier_Box0 : text_block = text_block{DefaultTextColor := White, DefaultShadowOffset:=ShadowOffset, DefaultShadowOpacity := DefaultTextShadowOpacity}

    #Box1
    TextureBlock_PetImages_Box1 : texture_block = texture_block{DefaultImage:=Images.EmptyTexture,DefaultDesiredSize:=PetsSelectorImageSize}
    TextBlock_Classes_Box1 : text_block = text_block{DefaultTextColor := White, DefaultShadowOffset:=ShadowOffset, DefaultShadowOpacity := DefaultTextShadowOpacity}
    TextBlock_PackA_Box1 : text_block = text_block{DefaultTextColor := White, DefaultShadowOffset:=ShadowOffset, DefaultShadowOpacity := DefaultTextShadowOpacity}
    TextBlock_PackB_Box1 : text_block = text_block{DefaultTextColor := White, DefaultShadowOffset:=ShadowOffset, DefaultShadowOpacity := DefaultTextShadowOpacity}
    TextBlock_Purity_Box1 : text_block = text_block{DefaultTextColor := White, DefaultShadowOffset:=ShadowOffset, DefaultShadowOpacity := DefaultTextShadowOpacity}
    TextBlock_Tier_Box1 : text_block = text_block{DefaultTextColor := White, DefaultShadowOffset:=ShadowOffset, DefaultShadowOpacity := DefaultTextShadowOpacity}

    #Box2
    TextureBlock_PetImages_Box2 : texture_block = texture_block{DefaultImage:=Images.EmptyTexture,DefaultDesiredSize:=PetsSelectorImageSize}
    TextBlock_Classes_Box2 : text_block = text_block{DefaultTextColor := White, DefaultShadowOffset:=ShadowOffset, DefaultShadowOpacity := DefaultTextShadowOpacity}
    TextBlock_PackA_Box2 : text_block = text_block{DefaultTextColor := White, DefaultShadowOffset:=ShadowOffset, DefaultShadowOpacity := DefaultTextShadowOpacity}
    TextBlock_PackB_Box2 : text_block = text_block{DefaultTextColor := White, DefaultShadowOffset:=ShadowOffset, DefaultShadowOpacity := DefaultTextShadowOpacity}
    TextBlock_Purity_Box2 : text_block = text_block{DefaultTextColor := White, DefaultShadowOffset:=ShadowOffset, DefaultShadowOpacity := DefaultTextShadowOpacity}
    TextBlock_Tier_Box2 : text_block = text_block{DefaultTextColor := White, DefaultShadowOffset:=ShadowOffset, DefaultShadowOpacity := DefaultTextShadowOpacity}

    #Box3
    TextureBlock_PetImages_Box3 : texture_block = texture_block{DefaultImage:=Images.EmptyTexture,DefaultDesiredSize:=PetsSelectorImageSize}
    TextBlock_Classes_Box3 : text_block = text_block{DefaultTextColor := White, DefaultShadowOffset:=ShadowOffset, DefaultShadowOpacity := DefaultTextShadowOpacity}
    TextBlock_PackA_Box3 : text_block = text_block{DefaultTextColor := White, DefaultShadowOffset:=ShadowOffset, DefaultShadowOpacity := DefaultTextShadowOpacity}
    TextBlock_PackB_Box3 : text_block = text_block{DefaultTextColor := White, DefaultShadowOffset:=ShadowOffset, DefaultShadowOpacity := DefaultTextShadowOpacity}
    TextBlock_Purity_Box3 : text_block = text_block{DefaultTextColor := White, DefaultShadowOffset:=ShadowOffset, DefaultShadowOpacity := DefaultTextShadowOpacity}
    TextBlock_Tier_Box3 : text_block = text_block{DefaultTextColor := White, DefaultShadowOffset:=ShadowOffset, DefaultShadowOpacity := DefaultTextShadowOpacity}

    #All Arrays
    var AllTextureBlock_PetImages : []texture_block = array{}
    var AllTextBlock_Classes : []text_block = array{}
    var AllTextBlock_PackA : []text_block = array{}
    var AllTextBlock_PackB : []text_block = array{}
    var AllTextBlock_Purity : []text_block = array{}
    var AllTextBlock_Tier : []text_block = array{}

    Text_PackageA_LV : text_block = text_block{DefaultTextColor := White, DefaultShadowOffset:=ShadowOffset, DefaultShadowOpacity := DefaultTextShadowOpacity}
    Text_PackageB_LV : text_block = text_block{DefaultTextColor := White, DefaultShadowOffset:=ShadowOffset, DefaultShadowOpacity := DefaultTextShadowOpacity}

    #COSTS
    TextureBlock_CostStatsA : texture_block = texture_block{DefaultImage:=Images.Coal_Icon,DefaultDesiredSize:=vector2{X:=SizeSmallIconX,Y:=SizeSmallIconY}}
    TextureBlock_CostStatsB : texture_block = texture_block{DefaultImage:=Images.Coal_Icon,DefaultDesiredSize:=vector2{X:=SizeSmallIconX,Y:=SizeSmallIconY}}

    TextBlock_CostStatsA : text_block = text_block{DefaultTextColor := MutantMaterialColor, DefaultShadowOffset:=ShadowOffset, DefaultShadowOpacity := DefaultTextShadowOpacity}
    TextBlock_CostStatsB : text_block = text_block{DefaultTextColor := MutantMaterialColor, DefaultShadowOffset:=ShadowOffset, DefaultShadowOpacity := DefaultTextShadowOpacity}

    #TOURNAMENT
    var Selected_TextureBlock_PetImages : texture_block = texture_block{DefaultImage:=Images.EmptyTexture,DefaultDesiredSize:=vector2{X:=75.0,Y:=75.0}}
    var Selected_TextBlock_Purity : text_block = text_block{DefaultTextColor := CurrentTextColor, DefaultShadowOffset:=ShadowOffset, DefaultShadowOpacity := DefaultTextShadowOpacity}
    var Selected_TextBlock_Class : text_block = text_block{DefaultTextColor := CurrentTextColor, DefaultShadowOffset:=ShadowOffset, DefaultShadowOpacity := DefaultTextShadowOpacity}
    var Selected_TextBlock_PackA : text_block = text_block{DefaultTextColor := CurrentTextColor, DefaultShadowOffset:=ShadowOffset, DefaultShadowOpacity := DefaultTextShadowOpacity}
    var Selected_TextBlock_PackB : text_block = text_block{DefaultTextColor := CurrentTextColor, DefaultShadowOffset:=ShadowOffset, DefaultShadowOpacity := DefaultTextShadowOpacity}

    #Package A
    var Selected_TextBlock_Current_TournamentHealth : text_block = text_block{DefaultTextColor := CurrentTextColor, DefaultShadowOffset:=ShadowOffset, DefaultShadowOpacity := DefaultTextShadowOpacity}
    var Selected_TextBlock_Current_TournamentDamage : text_block = text_block{DefaultTextColor := CurrentTextColor, DefaultShadowOffset:=ShadowOffset, DefaultShadowOpacity := DefaultTextShadowOpacity}
    var Selected_TextBlock_Current_TournamentSpeed : text_block = text_block{DefaultTextColor := CurrentTextColor, DefaultShadowOffset:=ShadowOffset, DefaultShadowOpacity := DefaultTextShadowOpacity}

    var Selected_TextBlock_Upgrade_TournamentHealth : text_block = text_block{DefaultTextColor := UpgradeTextColor, DefaultShadowOffset:=ShadowOffset, DefaultShadowOpacity := DefaultTextShadowOpacity}
    var Selected_TextBlock_Upgrade_TournamentDamage : text_block = text_block{DefaultTextColor := UpgradeTextColor, DefaultShadowOffset:=ShadowOffset, DefaultShadowOpacity := DefaultTextShadowOpacity}
    var Selected_TextBlock_Upgrade_TournamentSpeed : text_block = text_block{DefaultTextColor := UpgradeTextColor, DefaultShadowOffset:=ShadowOffset, DefaultShadowOpacity := DefaultTextShadowOpacity}

    #Package B
    var Selected_TextBlock_Current_TournamentActionSpeed : text_block = text_block{DefaultTextColor := CurrentTextColor, DefaultShadowOffset:=ShadowOffset, DefaultShadowOpacity := DefaultTextShadowOpacity}
    var Selected_TextBlock_Current_TournamentRange : text_block = text_block{DefaultTextColor := CurrentTextColor, DefaultShadowOffset:=ShadowOffset, DefaultShadowOpacity := DefaultTextShadowOpacity}
    var Selected_TextBlock_Current_TournamentHealingPerk : text_block = text_block{DefaultTextColor := CurrentTextColor, DefaultShadowOffset:=ShadowOffset, DefaultShadowOpacity := DefaultTextShadowOpacity}

    var Selected_TextBlock_Upgrade_TournamentActionSpeed : text_block = text_block{DefaultTextColor := UpgradeTextColor, DefaultShadowOffset:=ShadowOffset, DefaultShadowOpacity := DefaultTextShadowOpacity}
    var Selected_TextBlock_Upgrade_TournamentRange : text_block = text_block{DefaultTextColor := UpgradeTextColor, DefaultShadowOffset:=ShadowOffset, DefaultShadowOpacity := DefaultTextShadowOpacity}
    var Selected_TextBlock_Upgrade_TournamentHealingPerk : text_block = text_block{DefaultTextColor := UpgradeTextColor, DefaultShadowOffset:=ShadowOffset, DefaultShadowOpacity := DefaultTextShadowOpacity}

    #PLAYER STATS
        #Package A
    var Selected_TextBlock_Description_SpecialPerkA : text_block = text_block{DefaultTextColor := Pink, DefaultShadowOffset:=ShadowOffset, DefaultShadowOpacity := DefaultTextShadowOpacity}

    var Selected_TextBlock_Current_SpecialPerkA : text_block = text_block{DefaultTextColor := CurrentTextColor, DefaultShadowOffset:=ShadowOffset, DefaultShadowOpacity := DefaultTextShadowOpacity}
    var Selected_TextBlock_Current_PlayerPickDamage : text_block = text_block{DefaultTextColor := CurrentTextColor, DefaultShadowOffset:=ShadowOffset, DefaultShadowOpacity := DefaultTextShadowOpacity}
    var Selected_TextBlock_Current_PlayerWeakPoint : text_block = text_block{DefaultTextColor := CurrentTextColor, DefaultShadowOffset:=ShadowOffset, DefaultShadowOpacity := DefaultTextShadowOpacity}

    var Selected_TextBlock_Upgrade_SpecialPerkA : text_block = text_block{DefaultTextColor := UpgradeTextColor, DefaultShadowOffset:=ShadowOffset, DefaultShadowOpacity := DefaultTextShadowOpacity}
    var Selected_TextBlock_Upgrade_PlayerPickDamage : text_block = text_block{DefaultTextColor := UpgradeTextColor, DefaultShadowOffset:=ShadowOffset, DefaultShadowOpacity := DefaultTextShadowOpacity}
    var Selected_TextBlock_Upgrade_PlayerWeakPoint : text_block = text_block{DefaultTextColor := UpgradeTextColor, DefaultShadowOffset:=ShadowOffset, DefaultShadowOpacity := DefaultTextShadowOpacity}

        #Package B
    var Selected_TextBlock_Description_SpecialPerkB : text_block = text_block{DefaultTextColor := Pink}

    var Selected_TextBlock_Current_SpecialPerkB : text_block = text_block{DefaultTextColor := CurrentTextColor, DefaultShadowOffset:=ShadowOffset, DefaultShadowOpacity := DefaultTextShadowOpacity}
    var Selected_TextBlock_Current_PlayerPickEfficiency : text_block = text_block{DefaultTextColor := CurrentTextColor, DefaultShadowOffset:=ShadowOffset, DefaultShadowOpacity := DefaultTextShadowOpacity}
    var Selected_TextBlock_Current_PlayerPickCriticalDamage : text_block = text_block{DefaultTextColor := CurrentTextColor, DefaultShadowOffset:=ShadowOffset, DefaultShadowOpacity := DefaultTextShadowOpacity}

    var Selected_TextBlock_Upgrade_SpecialPerkB : text_block = text_block{DefaultTextColor := UpgradeTextColor, DefaultShadowOffset:=ShadowOffset, DefaultShadowOpacity := DefaultTextShadowOpacity}
    var Selected_TextBlock_Upgrade_PlayerPickEfficiency : text_block = text_block{DefaultTextColor := UpgradeTextColor, DefaultShadowOffset:=ShadowOffset, DefaultShadowOpacity := DefaultTextShadowOpacity}
    var Selected_TextBlock_Upgrade_PlayerPickCriticalDamage : text_block = text_block{DefaultTextColor := UpgradeTextColor, DefaultShadowOffset:=ShadowOffset, DefaultShadowOpacity := DefaultTextShadowOpacity}

    #Close
    CloseButton : button_loud = button_loud{}

    InitUI() : void =
        NavigateRightButton.SetText(BasicMessage("▶"))
        NavigateLeftButton.SetText(BasicMessage("◀"))
        CancelButton.SetText(BasicMessage("Cancel"))

        set AllTextureBlock_PetImages = array{TextureBlock_PetImages_Box0,TextureBlock_PetImages_Box1,texture_block{DefaultImage := Images.EmptyTexture},TextureBlock_PetImages_Box2,TextureBlock_PetImages_Box3}
        set AllTextBlock_Classes = array{TextBlock_Classes_Box0,TextBlock_Classes_Box1,text_block{},TextBlock_Classes_Box2,TextBlock_Classes_Box3}
        set AllTextBlock_PackA = array{TextBlock_PackA_Box0,TextBlock_PackA_Box1,text_block{},TextBlock_PackA_Box2,TextBlock_PackA_Box3}
        set AllTextBlock_PackB = array{TextBlock_PackB_Box0,TextBlock_PackB_Box1,text_block{},TextBlock_PackB_Box2,TextBlock_PackB_Box3}
        set AllTextBlock_Purity = array{TextBlock_Purity_Box0,TextBlock_Purity_Box1,text_block{},TextBlock_Purity_Box2,TextBlock_Purity_Box3}
        set AllTextBlock_Tier = array{TextBlock_Tier_Box0,TextBlock_Tier_Box1,text_block{},TextBlock_Tier_Box2,TextBlock_Tier_Box3}

        Subscribes()
        Canvas := MakeCanvas()
        set MaybeCanvas = option{Canvas}
        #Display_Canvas()

    SetAgent(InAgent : agent) : void =
        Remove_CanvasWithoutMessage()
        set MaybeAgent = option{InAgent}

    NavidateRight(Message : widget_message) : void =
        set CurrentOption = CurrentOption + 1
        if(Message.Player.IsActive[], ValidBakery:=PetsManager.GameManager.GetAgentBakery_Transacted(agent[Message.Player])?, ValidBakery.PlayerDataCache.AllPets.Length <= CurrentOption):
           set CurrentOption = ValidBakery.PlayerDataCache.AllPets.Length-1
            #Print("ERROR, no more pets on this side")
        RefreshAll()

    NavidateLeft(Message : widget_message) : void =
        set CurrentOption = CurrentOption - 1
        if(CurrentOption < 0):
            set CurrentOption = 0
            #Print("ERROR, no more pets on this side")
        RefreshAll()

           
    Subscribes() : void =
        Upgrade_AStatsButton.OnClick().Subscribe(TryUpgradeAStats)
        Upgrade_BStatsButton.OnClick().Subscribe(TryUpgradeBStats)
        NavigateRightButton.OnClick().Subscribe(NavidateRight)
        NavigateLeftButton.OnClick().Subscribe(NavidateLeft)
        CancelButton.OnClick().Subscribe(handler_hexgame{MaybePetsManager:=option{PetsManager}, SubMenuOption:=SUBMENU_IMPROVE_PET}.HandlerUIFromEverythingToMainMenu)

    RefreshAll() : void =
        if(Agent := MaybeAgent?,Player := player[Agent], Player.IsActive[], ValidBakery:=PetsManager.GameManager.GetAgentBakery_Transacted(Agent)?):
            TotalCoal_TextBlock.SetText(BasicMessage(GetMagnitudeLetter(ValidBakery.PlayerDataCache.Inventory_Coal, "", "")))

            AllPets:=ValidBakery.PlayerDataCache.AllPets
            if(Pet := AllPets[CurrentOption]):

                Text_PackageA_LV.SetText(BasicMessage("PACKAGE A - LEVEL {Pet.LVPet_PackageA}"))
                Text_PackageB_LV.SetText(BasicMessage("PACKAGE B - LEVEL {Pet.LVPet_PackageB}"))

                var SelectedEmoji_SpecialA : string = ""
                var SelectedEmoji_SpecialB : string = ""

                #Package A
                TupleMaxHealth := CalculePetStat(STAT_TYPE_HEALTH,Pet.LVPet_PackageA,Pet.Type,Pet.Purity)
                TupleAttackDamage := CalculePetStat(STAT_TYPE_DAMAGE,Pet.LVPet_PackageA,Pet.Type,Pet.Purity)
                TupleSpeed := CalculePetStat(STAT_TYPE_SPEED,Pet.LVPet_PackageA,Pet.Type,Pet.Purity)

                TupleSpecialA := CalculePetStat(STAT_TYPE_PLAYER_SPECIAL_A,Pet.LVPet_PackageA,Pet.Type,Pet.Purity)
                TuplePickDamage := CalculePetStat(STAT_TYPE_PLAYER_PICKDAMAGE,Pet.LVPet_PackageA,Pet.Type,Pet.Purity)
                TupleWeakPoint := CalculePetStat(STAT_TYPE_PLAYER_WEAKPOINTMULT,Pet.LVPet_PackageA,Pet.Type,Pet.Purity)

                #Package B
                TupleAttackSpeed := CalculePetStat(STAT_TYPE_ATTACKSPEED,Pet.LVPet_PackageB,Pet.Type,Pet.Purity)
                TupleRange := CalculePetStat(STAT_TYPE_RANGE,Pet.LVPet_PackageB,Pet.Type,Pet.Purity)
                TupleHealingPerk := CalculePetStat(STAT_TYPE_HEALING,Pet.LVPet_PackageB,Pet.Type,Pet.Purity)

                TupleSpecialB := CalculePetStat(STAT_TYPE_PLAYER_SPECIAL_B,Pet.LVPet_PackageB,Pet.Type,Pet.Purity)
                TuplePickEfficiency := CalculePetStat(STAT_TYPE_PLAYER_PICKEFFICENCY,Pet.LVPet_PackageB,Pet.Type,Pet.Purity)
                TuplePickCriticalDamage := CalculePetStat(STAT_TYPE_PLAYER_PICKCRITICALMULT,Pet.LVPet_PackageB,Pet.Type,Pet.Purity)

                #NextLevel
                NextMaxHealth := TupleMaxHealth(2)
                NextAttackDamage := TupleAttackDamage(2)
                NextSpeed := TupleSpeed(2)

                #NextLevel
                NextSpecialA := TupleSpecialA(2)
                NextPickDamage := TuplePickDamage(2)
                NextWeakPoint := TupleWeakPoint(2)

                #NextLevel
                NextActionSpeed := TupleAttackSpeed(2)
                NextRange := TupleRange(2)
                NextHealingPerk := TupleHealingPerk(2)

                #NextLevel
                NextSpecialB := TupleSpecialB(2)
                NextPickEfficiency := TuplePickEfficiency(2)
                NextCriticalMulti := TuplePickCriticalDamage(2)

                if(Pet.Type=ADC):
                    set SelectedEmoji_SpecialA = "🎯"
                    set SelectedEmoji_SpecialB = "🏹"
                    Selected_TextBlock_Current_SpecialPerkA.SetText(BasicMessage(GetTripleDecimalMagnitudeLetter(TupleSpecialA(0), " {SelectedEmoji_SpecialA} x", "")))
                    Selected_TextBlock_Current_SpecialPerkB.SetText(BasicMessage(GetDoubleDecimalMagnitudeLetter(TupleSpecialB(0), " {SelectedEmoji_SpecialB}", "s")))
                    Selected_TextBlock_Upgrade_SpecialPerkA.SetText(BasicMessage(GetTripleDecimalMagnitudeLetter(NextSpecialA, "+", "")))
                    Selected_TextBlock_Upgrade_SpecialPerkB.SetText(BasicMessage(GetDoubleDecimalMagnitudeLetter(NextSpecialB, "", "s")))
                    Selected_TextBlock_Description_SpecialPerkA.SetText(BasicMessage("Bow damage:"))
                    Selected_TextBlock_Description_SpecialPerkB.SetText(BasicMessage("Arrow craft:"))
                    
                else if(Pet.Type=DRUID):
                    set SelectedEmoji_SpecialA = "💚"
                    set SelectedEmoji_SpecialB = "💉"
                    Selected_TextBlock_Current_SpecialPerkA.SetText(BasicMessage(GetTripleDecimalMagnitudeLetter(TupleSpecialA(0), "{SelectedEmoji_SpecialA} ", "")))
                    Selected_TextBlock_Current_SpecialPerkB.SetText(BasicMessage(GetTripleDecimalMagnitudeLetter(TupleSpecialB(0), "{SelectedEmoji_SpecialB} ", "")))
                    Selected_TextBlock_Upgrade_SpecialPerkA.SetText(BasicMessage(GetTripleDecimalMagnitudeLetter(NextSpecialA, "+", "")))
                    Selected_TextBlock_Upgrade_SpecialPerkB.SetText(BasicMessage(GetTripleDecimalMagnitudeLetter(NextSpecialB, "+", "")))
                    Selected_TextBlock_Description_SpecialPerkA.SetText(BasicMessage("Pasive heal:"))
                    Selected_TextBlock_Description_SpecialPerkB.SetText(BasicMessage("Emergency heal:"))
                else if(Pet.Type=TANK):
                    set SelectedEmoji_SpecialA = "💗"
                    set SelectedEmoji_SpecialB = "🛡️"
                    Selected_TextBlock_Current_SpecialPerkA.SetText(BasicMessage(GetTripleDecimalMagnitudeLetter(TupleSpecialA(0), "{SelectedEmoji_SpecialA} ", "")))
                    Selected_TextBlock_Current_SpecialPerkB.SetText(BasicMessage(GetTripleDecimalMagnitudeLetter(TupleSpecialB(0), "{SelectedEmoji_SpecialB} ", "")))
                    Selected_TextBlock_Upgrade_SpecialPerkA.SetText(BasicMessage(GetTripleDecimalMagnitudeLetter(NextSpecialA, "+", "")))
                    Selected_TextBlock_Upgrade_SpecialPerkB.SetText(BasicMessage(GetTripleDecimalMagnitudeLetter(NextSpecialB, "+", "")))
                    Selected_TextBlock_Description_SpecialPerkA.SetText(BasicMessage("Extra health:"))
                    Selected_TextBlock_Description_SpecialPerkB.SetText(BasicMessage("Combat shield:"))
                else if(Pet.Type=FIGHTER):
                    set SelectedEmoji_SpecialA = "🔪"
                    set SelectedEmoji_SpecialB = "🗡️"
                    Selected_TextBlock_Current_SpecialPerkA.SetText(BasicMessage(GetTripleDecimalMagnitudeLetter(TupleSpecialA(0), "{SelectedEmoji_SpecialA} x", "")))
                    Selected_TextBlock_Current_SpecialPerkB.SetText(BasicMessage(GetTripleDecimalMagnitudeLetter(TupleSpecialB(0), "{SelectedEmoji_SpecialB} x", "")))
                    Selected_TextBlock_Upgrade_SpecialPerkA.SetText(BasicMessage(GetTripleDecimalMagnitudeLetter(NextSpecialA, "+", "")))
                    Selected_TextBlock_Upgrade_SpecialPerkB.SetText(BasicMessage(GetTripleDecimalMagnitudeLetter(NextSpecialB, "+", "")))
                    Selected_TextBlock_Description_SpecialPerkA.SetText(BasicMessage("Melee damage:"))
                    Selected_TextBlock_Description_SpecialPerkB.SetText(BasicMessage("Melee UltraStrike:"))
                else if(Pet.Type=HUNTER):
                    set SelectedEmoji_SpecialA = "🐾"
                    set SelectedEmoji_SpecialB = "💢"
                    Selected_TextBlock_Current_SpecialPerkA.SetText(BasicMessage(GetTripleDecimalMagnitudeLetter(TupleSpecialA(0), "{SelectedEmoji_SpecialA} x", "")))
                    Selected_TextBlock_Current_SpecialPerkB.SetText(BasicMessage(GetTripleDecimalMagnitudeLetter(TupleSpecialB(0), "{SelectedEmoji_SpecialB} x", "")))
                    Selected_TextBlock_Upgrade_SpecialPerkA.SetText(BasicMessage(GetTripleDecimalMagnitudeLetter(NextSpecialA, "+", "")))
                    Selected_TextBlock_Upgrade_SpecialPerkB.SetText(BasicMessage(GetTripleDecimalMagnitudeLetter(NextSpecialB, "+", "")))
                    Selected_TextBlock_Description_SpecialPerkA.SetText(BasicMessage("Hunt probability:"))
                    Selected_TextBlock_Description_SpecialPerkB.SetText(BasicMessage("Weapon critical:"))
                    

                TuplePurity:=ConvertRawPurity(Pet.Purity)
                SetPetImage(Pet.Type, Selected_TextureBlock_PetImages)
                SetPetPurityAndTier(TuplePurity(1),TuplePurity(0),Selected_TextBlock_Purity)
                SetPetClass(Pet,Selected_TextBlock_Class,Selected_TextBlock_PackA,Selected_TextBlock_PackB)

                #ALL STATS
                #Package A
                Selected_TextBlock_Current_TournamentHealth.SetText(BasicMessage(GetDoubleDecimalMagnitudeLetter(TupleMaxHealth(0), "❤️ ", "")))
                Selected_TextBlock_Current_TournamentDamage.SetText(BasicMessage(GetDoubleDecimalMagnitudeLetter(TupleAttackDamage(0), "💥 ", "")))
                Selected_TextBlock_Current_TournamentSpeed.SetText(BasicMessage(GetDoubleDecimalMagnitudeLetter(TupleSpeed(0), "🏃 ", "s / tile")))

                Selected_TextBlock_Current_PlayerPickDamage.SetText(BasicMessage(GetTripleDecimalMagnitudeLetter(TuplePickDamage(0), "🔨 x", "")))
                Selected_TextBlock_Current_PlayerWeakPoint.SetText(BasicMessage(GetTripleDecimalMagnitudeLetter(TupleWeakPoint(0), "🧨 x", "")))

                    #NEXT UPGRADE
                if(Pet.LVPet_PackageA>=MaxLV_Pet_PackageA):
                    TextBlock_CostStatsA.SetTextColor(MaxedColor)
                    TextBlock_CostStatsA.SetText(BasicMessage("MAXED"))

                    Selected_TextBlock_Upgrade_TournamentHealth.SetText(BasicMessage("MAXED"))
                    Selected_TextBlock_Upgrade_TournamentDamage.SetText(BasicMessage("MAXED"))
                    Selected_TextBlock_Upgrade_TournamentSpeed.SetText(BasicMessage("MAXED"))
                    Selected_TextBlock_Upgrade_TournamentHealth.SetTextColor(MaxedColor)
                    Selected_TextBlock_Upgrade_TournamentSpeed.SetTextColor(MaxedColor)
                    Selected_TextBlock_Upgrade_TournamentDamage.SetTextColor(MaxedColor)

                    Selected_TextBlock_Upgrade_SpecialPerkA.SetText(BasicMessage("MAXED"))
                    Selected_TextBlock_Upgrade_PlayerPickDamage.SetText(BasicMessage("MAXED"))
                    Selected_TextBlock_Upgrade_PlayerWeakPoint.SetText(BasicMessage("MAXED"))
                    Selected_TextBlock_Upgrade_SpecialPerkA.SetTextColor(MaxedColor)
                    Selected_TextBlock_Upgrade_PlayerPickDamage.SetTextColor(MaxedColor)
                    Selected_TextBlock_Upgrade_PlayerWeakPoint.SetTextColor(MaxedColor)
                else:
                    Selected_TextBlock_Upgrade_TournamentHealth.SetTextColor(UpgradeTextColor)
                    Selected_TextBlock_Upgrade_TournamentSpeed.SetTextColor(UpgradeTextColor)
                    Selected_TextBlock_Upgrade_TournamentDamage.SetTextColor(UpgradeTextColor)

                    Selected_TextBlock_Upgrade_SpecialPerkA.SetTextColor(UpgradeTextColor)
                    Selected_TextBlock_Upgrade_PlayerPickDamage.SetTextColor(UpgradeTextColor)
                    Selected_TextBlock_Upgrade_PlayerWeakPoint.SetTextColor(UpgradeTextColor)

                    Selected_TextBlock_Upgrade_TournamentHealth.SetText(BasicMessage(GetDoubleDecimalMagnitudeLetter(NextMaxHealth, "+", "")))
                    Selected_TextBlock_Upgrade_TournamentDamage.SetText(BasicMessage(GetDoubleDecimalMagnitudeLetter(NextAttackDamage, "+", "")))
                    Selected_TextBlock_Upgrade_TournamentSpeed.SetText(BasicMessage(GetDoubleDecimalMagnitudeLetter(NextSpeed, "", "s")))

                    Selected_TextBlock_Upgrade_PlayerPickDamage.SetText(BasicMessage(GetTripleDecimalMagnitudeLetter(NextPickDamage, "+", "")))
                    Selected_TextBlock_Upgrade_PlayerWeakPoint.SetText(BasicMessage(GetTripleDecimalMagnitudeLetter(NextWeakPoint, "+", "")))

                    UpgradeCost:=CalculePetStat(STAT_TYPE_HEALTH,Pet.LVPet_PackageA,Pet.Type,Pet.Purity)(1)
                    TextBlock_CostStatsA.SetText(BasicMessage(GetMagnitudeLetter(UpgradeCost, "x ","")))
                    if(ValidBakery.PlayerDataCache.Inventory_Coal<UpgradeCost):
                        TextBlock_CostStatsA.SetTextColor(NotEnoughTextColor)
                    else:
                        TextBlock_CostStatsA.SetTextColor(CoalTextColor)
                    

                
               
                Selected_TextBlock_Current_TournamentActionSpeed.SetText(BasicMessage(GetDoubleDecimalMagnitudeLetter(TupleAttackSpeed(0), "🔥", "s")))
                Selected_TextBlock_Current_TournamentRange.SetText(BasicMessage(GetMagnitudeLetter(TupleRange(0), "🏹 ", " tiles")))

                if(Pet.Type = DRUID):
                    Selected_TextBlock_Current_TournamentHealingPerk.SetText(BasicMessage(GetDoubleDecimalMagnitudeLetter(TupleHealingPerk(0), "💉 ", "")))
                else:
                    Selected_TextBlock_Current_TournamentHealingPerk.SetText(BasicMessage("-"))

                Selected_TextBlock_Current_PlayerPickEfficiency.SetText(BasicMessage(GetTripleDecimalMagnitudeLetter(TuplePickEfficiency(0), "🚀 x", "")))
                Selected_TextBlock_Current_PlayerPickCriticalDamage.SetText(BasicMessage(GetTripleDecimalMagnitudeLetter(TuplePickCriticalDamage(0), "💥 x", "")))
                    

                if(Pet.LVPet_PackageB>=MaxLV_Pet_PackageB):
                    TextBlock_CostStatsB.SetTextColor(MaxedColor)
                    TextBlock_CostStatsB.SetText(BasicMessage("MAXED"))

                    Selected_TextBlock_Upgrade_TournamentActionSpeed.SetText(BasicMessage("MAXED"))
                    Selected_TextBlock_Upgrade_TournamentRange.SetText(BasicMessage("MAXED"))
                    Selected_TextBlock_Upgrade_TournamentHealingPerk.SetText(BasicMessage("MAXED"))

                    Selected_TextBlock_Upgrade_TournamentActionSpeed.SetTextColor(MaxedColor)
                    Selected_TextBlock_Upgrade_TournamentRange.SetTextColor(MaxedColor)
                    Selected_TextBlock_Upgrade_TournamentHealingPerk.SetTextColor(MaxedColor)

                    Selected_TextBlock_Upgrade_SpecialPerkB.SetText(BasicMessage("MAXED"))
                    Selected_TextBlock_Upgrade_PlayerPickEfficiency.SetText(BasicMessage("MAXED"))
                    Selected_TextBlock_Upgrade_PlayerPickCriticalDamage.SetText(BasicMessage("MAXED"))

                    Selected_TextBlock_Upgrade_SpecialPerkB.SetTextColor(MaxedColor)
                    Selected_TextBlock_Upgrade_PlayerPickEfficiency.SetTextColor(MaxedColor)
                    Selected_TextBlock_Upgrade_PlayerPickCriticalDamage.SetTextColor(MaxedColor)
                else:
                    Selected_TextBlock_Upgrade_TournamentActionSpeed.SetTextColor(UpgradeTextColor)
                    Selected_TextBlock_Upgrade_TournamentRange.SetTextColor(UpgradeTextColor)
                    Selected_TextBlock_Upgrade_TournamentHealingPerk.SetTextColor(UpgradeTextColor)

                    Selected_TextBlock_Upgrade_SpecialPerkB.SetTextColor(UpgradeTextColor)
                    Selected_TextBlock_Upgrade_PlayerPickEfficiency.SetTextColor(UpgradeTextColor)
                    Selected_TextBlock_Upgrade_PlayerPickCriticalDamage.SetTextColor(UpgradeTextColor)

                    Selected_TextBlock_Upgrade_TournamentActionSpeed.SetText(BasicMessage(GetDoubleDecimalMagnitudeLetter(NextActionSpeed, "", "s")))
                    Selected_TextBlock_Upgrade_TournamentRange.SetText(BasicMessage(GetMagnitudeLetter(NextRange, "+", "")))
                    Selected_TextBlock_Upgrade_TournamentHealingPerk.SetText(BasicMessage(GetDoubleDecimalMagnitudeLetter(NextHealingPerk, "+", "")))

                    if(Pet.Type = DRUID):
                        Selected_TextBlock_Upgrade_TournamentHealingPerk.SetText(BasicMessage(GetDoubleDecimalMagnitudeLetter(NextHealingPerk, "+", "")))
                    else:
                        Selected_TextBlock_Upgrade_TournamentHealingPerk.SetText(BasicMessage("-"))

                    Selected_TextBlock_Upgrade_PlayerPickEfficiency.SetText(BasicMessage(GetTripleDecimalMagnitudeLetter(NextPickEfficiency, "+", "")))
                    Selected_TextBlock_Upgrade_PlayerPickCriticalDamage.SetText(BasicMessage(GetTripleDecimalMagnitudeLetter(NextCriticalMulti, "+", "")))
                    UpgradeCost:=CalculePetStat(STAT_TYPE_HEALTH,Pet.LVPet_PackageB,Pet.Type,Pet.Purity)(1)
                    TextBlock_CostStatsB.SetText(BasicMessage(GetMagnitudeLetter(UpgradeCost, "x ","")))
                    TextBlock_CostStatsB.SetTextColor(CoalTextColor)
                    if(ValidBakery.PlayerDataCache.Inventory_Coal<UpgradeCost):
                        TextBlock_CostStatsB.SetTextColor(NotEnoughTextColor)
                    else:
                        TextBlock_CostStatsB.SetTextColor(CoalTextColor)

            #The other images
            for(Index:=0..AllTextureBlock_PetImages.Length-1):
                if:
                    TextureBlock_PetImage:=AllTextureBlock_PetImages[Index]
                    TextBlock_Class:=AllTextBlock_Classes[Index]
                    TextBlock_PackA:=AllTextBlock_PackA[Index]
                    TextBlock_PackB:=AllTextBlock_PackB[Index]
                    TextBlock_Purity:=AllTextBlock_Purity[Index]
                    TextBlock_Tier:=AllTextBlock_Tier[Index]
                then:
                    if(PetIndexSelected:=AllPets[CurrentOption-2+Index]):
                        TuplePurity:=ConvertRawPurity(PetIndexSelected.Purity)
                        SetPetImage(PetIndexSelected.Type, TextureBlock_PetImage)
                        SetPetClass(PetIndexSelected, TextBlock_Class, TextBlock_PackA, TextBlock_PackB)
                        SetPetPurity(TuplePurity(1), TextBlock_Purity)
                        SetPetTier(TuplePurity(0), TextBlock_Tier)
                    else:
                        TextureBlock_PetImage.SetImage(Images.Cross_Icon)
                        TextBlock_Class.SetText(BasicMessage("-"))
                        TextBlock_PackA.SetText(BasicMessage("-"))
                        TextBlock_PackA.SetTextColor(White)
                        TextBlock_PackB.SetText(BasicMessage("-"))
                        TextBlock_PackB.SetTextColor(White)
                        TextBlock_Purity.SetText(BasicMessage("-"))
                        TextBlock_Tier.SetText(BasicMessage("-"))
                        TextureBlock_PetImage.SetTint(White)
                        TextBlock_Class.SetTextColor(White)
                        TextBlock_Purity.SetTextColor(White)
                        TextBlock_Tier.SetTextColor(White)
                        #Pet not found!
        PetsManager.Generate_PlayerEnhacements()
                
    TryUpgradeAStats(Message : widget_message) : void =
        if(Agent := MaybeAgent?, Player := player[Agent], Player.IsActive[], ValidBakery:=PetsManager.GameManager.GetAgentBakery_Transacted(agent[Message.Player])?):
            PetsManager.GameManager.VFXManager.SmallObjetive_AudioPlayer.Play(Agent)
            var AllPets : []pet_data_table = ValidBakery.PlayerDataCache.AllPets
            if(Pet := AllPets[CurrentOption]):
                if(Pet.LVPet_PackageA<MaxLV_Pet_PackageA):
                    UpgradeCost:=CalculePetStat(STAT_TYPE_HEALTH,Pet.LVPet_PackageA,Pet.Type,Pet.Purity)(1)
                    if(UpgradeCost>ValidBakery.PlayerDataCache.Inventory_Coal):
                        PetsManager.GameManager.VFXManager.ErrorSound_AudioPlayer.Play(Agent)
                        #PrintWarning("TO DO NOT ENOUGH Coal TryUpgradeAStats pet")
                    else:
                        TableWithoutCoal:=PetsManager.GameManager.PlayersDataManager.WithdrawResource(ValidBakery.PlayerDataCache,UpgradeCost,COAL)
                        NewPet:=PetsManager.GameManager.PlayersDataManager.SetPetPackA_LV(Pet,Pet.LVPet_PackageA+1)
                        if:
                            set AllPets[CurrentOption] = NewPet
                        then:
                            FinalTable:=PetsManager.GameManager.PlayersDataManager.SavePets(TableWithoutCoal,AllPets)
                            #PetsManager.GameManager.PlayersDataManager.SaveNewData(Player,PlayerData,FinalTable)
                            MaybeBakery:=PetsManager.GameManager.GetAgentBakery(Agent)
                            ValidBakery.SetNewPlayerDataCache(FinalTable)
                            RefreshAll()
                            ValidBakery.CheckQuestProgress(1.0,IMPROVEPET_TUTORIAL_TYPE)
                            spawn:
                                ValidBakery.GameManager.CalculateAndSaveFame(Agent)
            PetsManager.Generate_PlayerEnhacements()
            PetsManager.GameManager.RefreshInventoryUI(Agent)

    TryUpgradeBStats(Message : widget_message) : void =
        if(Agent := MaybeAgent?,Player := player[Agent], Player.IsActive[], ValidBakery:=PetsManager.GameManager.GetAgentBakery_Transacted(agent[Message.Player])?):
            PetsManager.GameManager.VFXManager.SmallObjetive_AudioPlayer.Play(Agent)
            var AllPets : []pet_data_table = ValidBakery.PlayerDataCache.AllPets
            if(Pet := AllPets[CurrentOption]):
                if(Pet.LVPet_PackageB<MaxLV_Pet_PackageB):
                    UpgradeCost:=CalculePetStat(STAT_TYPE_HEALTH,Pet.LVPet_PackageB,Pet.Type,Pet.Purity)(1)
                    if(UpgradeCost>ValidBakery.PlayerDataCache.Inventory_Coal):
                        PrintWarning("TO DO NOT ENOUGH Coal TryUpgradeAStats pet")
                        PetsManager.GameManager.VFXManager.ErrorSound_AudioPlayer.Play(Agent)
                    else:
                        TableWithoutCoal:=PetsManager.GameManager.PlayersDataManager.WithdrawResource(ValidBakery.PlayerDataCache,UpgradeCost,COAL)
                        NewPet:=PetsManager.GameManager.PlayersDataManager.SetPetPackB_LV(Pet,Pet.LVPet_PackageB+1)
                        if:
                            set AllPets[CurrentOption] = NewPet
                        then:
                            FinalTable:=PetsManager.GameManager.PlayersDataManager.SavePets(TableWithoutCoal,AllPets)
                            ValidBakery.SetNewPlayerDataCache(FinalTable)
                            #PetsManager.GameManager.PlayersDataManager.SaveNewData(Player,PlayerData,FinalTable)
                            RefreshAll()
                            ValidBakery.CheckQuestProgress(1.0,IMPROVEPET_TUTORIAL_TYPE)
                            spawn:
                                ValidBakery.GameManager.CalculateAndSaveFame(Agent)
            PetsManager.Generate_PlayerEnhacements()
            PetsManager.GameManager.RefreshInventoryUI(Agent)

    Remove_Canvas(Message : widget_message) : void =
        set UIIsDisplayed = false
        if:
            Agent :=MaybeAgent?
            Player:=player[Agent]
            PlayerUI:=GetPlayerUI[Player]
            Canvas := MaybeCanvas?
        then:
            PlayerUI.RemoveWidget(Canvas)
            RemoveDelay_OnRemove()

    
    GoToNavigation(Message : widget_message) : void =
        set UIIsDisplayed = false
        if:
            Agent :=MaybeAgent?
            Player:=player[Agent]
            PlayerUI:=GetPlayerUI[Player]
            Canvas := MaybeCanvas?
        then:
            PlayerUI.RemoveWidget(Canvas)
            RemoveDelay_OnRemove()


    Remove_CanvasWithoutMessage() : void =
        set UIIsDisplayed = false
        if:
            Agent :=MaybeAgent?
            Player:=player[Agent]
            PlayerUI:=GetPlayerUI[Player]
            Canvas := MaybeCanvas?
        then:
            PlayerUI.RemoveWidget(Canvas)
            RemoveDelay_OnRemove()

    Display_Canvas() : void =
        set UIIsDisplayed = true
        var Index : int = 0 
        if:
            Agent :=MaybeAgent?
            PlayerUI:=GetPlayerUI[player[Agent]]
            Canvas := MaybeCanvas?
        then:
            PlayerUI.AddWidget(Canvas,player_ui_slot{ZOrder:=0, InputMode := ui_input_mode.All})
            RemoveDelay_OnAdd()
            RefreshAll()

    RemoveDelay_OnAdd() : void =
        CloseButton.SetText(BasicMessage("Close"))

    RemoveDelay_OnRemove() : void =
        CloseButton.SetText(BasicMessage("Close "))
  
    
   

    MakeCanvas(): canvas=
        Canvas := canvas:
            Slots:=array:
                #Main overlay
                canvas_slot:
                    Anchors:=anchors{Minimum:=vector2{X:=0.5,Y:=0.5},Maximum:=vector2{X:=0.5,Y:=0.5}}
                    ZOrder:=0
                    SizeToContent:=true
                    Offsets:=margin{Left:=0.0,Top:=0.0,Right:=0.0,Bottom:=0.0} #PositionX, PositionY, SizeX, SizeY
                    Alignment:=vector2{X:=0.5,Y:=0.5}
                    Widget:=texture_block{DefaultDesiredSize:=vector2{X:=3000.0,Y:=1250.0},DefaultImage:=Images.Img_MainBackground}

                canvas_slot:
                    Anchors:=anchors{Minimum:=vector2{X:=1.0,Y:=0.0},Maximum:=vector2{X:=1.0,Y:=0.0}}
                    ZOrder:=0
                    SizeToContent:=true
                    Offsets:=margin{Left:=-10.0,Top:=10.0,Right:=0.0,Bottom:=0.0} #PositionX, PositionY, SizeX, SizeY
                    Alignment:=vector2{X:=1.0,Y:=0.0}
                    Widget:=stack_box:  
                        Orientation:=orientation.Vertical
                            Slots:=array:
                                stack_box_slot:
                                    HorizontalAlignment:=horizontal_alignment.Right
                                    VerticalAlignment:=vertical_alignment.Center
                                    Widget:=overlay:
                                        Slots:=array: 
                                            overlay_slot:
                                                Widget:=texture_block{DefaultDesiredSize:=vector2{X:=SizeBoxSaphiresX,Y:=SizeBoxSaphiresY},DefaultImage:=Images.EmptyTexture} 
                                            overlay_slot:
                                                HorizontalAlignment:=horizontal_alignment.Fill
                                                VerticalAlignment:=vertical_alignment.Fill
                                                Widget:=texture_block{DefaultImage:=Images.Img_LineBackground}  
                                            overlay_slot:
                                                HorizontalAlignment:=horizontal_alignment.Left
                                                VerticalAlignment:=vertical_alignment.Center
                                                Padding:=margin{Left:=Icon_LeftMargin}
                                                Widget:=texture_block{DefaultDesiredSize:=vector2{X:=IconSmallSizeX,Y:=IconSmallSizeY},DefaultImage:=Images.Coal_Icon}
                                            overlay_slot:
                                                HorizontalAlignment:=horizontal_alignment.Left
                                                VerticalAlignment:=vertical_alignment.Center
                                                Padding:=margin{Left:=TextIcon_LeftMargin}
                                                Widget:=TotalCoal_TextBlock  
                                    
                canvas_slot:
                    Anchors:=anchors{Minimum:=vector2{X:=0.5,Y:=0.5},Maximum:=vector2{X:=0.5,Y:=0.5}}
                    ZOrder:=0
                    SizeToContent:=true
                    Offsets:=margin{Left:=0.0,Top:=-75.0,Right:=0.0,Bottom:=0.0} #PositionX, PositionY, SizeX, SizeY
                    Alignment:=vector2{X:=0.5,Y:=0.5}
                    Widget:=overlay:
                        Slots:=array:
                            overlay_slot:
                                HorizontalAlignment:=horizontal_alignment.Center
                                VerticalAlignment:=vertical_alignment.Center
                                Widget:=stack_box:  
                                    Orientation:=orientation.Horizontal
                                        Slots:=array:
                                            stack_box_slot: #Slot 0
                                                Padding:=margin{Right:=SuperLowPadding,Left:=SuperLowPadding}
                                                HorizontalAlignment:=horizontal_alignment.Center
                                                VerticalAlignment:=vertical_alignment.Bottom
                                                Widget:=stack_box:  
                                                    Orientation:=orientation.Vertical
                                                        Slots:=array:
                                                            stack_box_slot: 
                                                                Padding:=margin{Top:=SelectorBoxPadding,Bottom:=SelectorBoxPadding}
                                                                Widget:=overlay:
                                                                    Slots:=array:
                                                                        overlay_slot:
                                                                            HorizontalAlignment:=horizontal_alignment.Fill
                                                                            VerticalAlignment:=vertical_alignment.Fill
                                                                            Widget:=texture_block{DefaultImage:=Images.Img_LineBackground}
                                                                        overlay_slot:
                                                                            HorizontalAlignment:=horizontal_alignment.Fill
                                                                            VerticalAlignment:=vertical_alignment.Fill
                                                                            Padding:=margin{Left:=SelectorBoxPadding,Right:=SelectorBoxPadding,Top:=SelectorBoxPadding,Bottom:=SelectorBoxPadding}
                                                                            Widget:=texture_block{DefaultImage:=Images.Img_LineBackground}
                                                                        overlay_slot:
                                                                            HorizontalAlignment:=horizontal_alignment.Center
                                                                            VerticalAlignment:=vertical_alignment.Center
                                                                            Widget:=texture_block{DefaultImage:=Images.EmptyTexture,DefaultDesiredSize:=vector2{X:=258.0,Y:=300.0}}
                                                                        overlay_slot:
                                                                            HorizontalAlignment:=horizontal_alignment.Center
                                                                            VerticalAlignment:=vertical_alignment.Center
                                                                            Widget:=stack_box:
                                                                                Orientation:=orientation.Vertical
                                                                                Slots:=array:
                                                                                    stack_box_slot: 
                                                                                        Padding:=RowPadding
                                                                                        HorizontalAlignment:=horizontal_alignment.Center
                                                                                        VerticalAlignment:=vertical_alignment.Center
                                                                                        Widget:=TextureBlock_PetImages_Box0
                                                                                    stack_box_slot: 
                                                                                        HorizontalAlignment:=horizontal_alignment.Center
                                                                                        VerticalAlignment:=vertical_alignment.Center
                                                                                        #Padding:=margin{Left:=TextPetsBoxPadding}
                                                                                        Widget:=stack_box:
                                                                                            Orientation:=orientation.Horizontal
                                                                                            Slots:=array:
                                                                                                stack_box_slot: 
                                                                                                    Padding:=margin{Left:=2.5,Right:=2.5}
                                                                                                    Widget:=TextBlock_Classes_Box0
                                                                                                stack_box_slot: 
                                                                                                    Padding:=margin{Left:=2.5, Right:=1.0}
                                                                                                    Widget:=TextBlock_PackA_Box0
                                                                                                stack_box_slot: 
                                                                                                    Padding:=margin{Right:=2.5}
                                                                                                    Widget:=TextBlock_PackB_Box0
                                                                                    stack_box_slot: 
                                                                                        Padding:=RowPadding
                                                                                        HorizontalAlignment:=horizontal_alignment.Center
                                                                                        VerticalAlignment:=vertical_alignment.Center
                                                                                        Widget:=TextBlock_Purity_Box0
                                                                                    stack_box_slot: 
                                                                                        Padding:=RowPadding
                                                                                        HorizontalAlignment:=horizontal_alignment.Center
                                                                                        VerticalAlignment:=vertical_alignment.Center
                                                                                        Widget:=TextBlock_Tier_Box0                                    
                                                            stack_box_slot: 
                                                                Widget:=overlay:
                                                                    Slots:=array:
                                                                        overlay_slot:
                                                                            HorizontalAlignment:=horizontal_alignment.Fill
                                                                            VerticalAlignment:=vertical_alignment.Fill
                                                                            Widget:=texture_block{DefaultImage:=Images.EmptyTexture,DefaultDesiredSize:=QuietButton_PetSelectorDesiredSize}
                                            stack_box_slot: #Slot 1
                                                Padding:=margin{Right:=SuperLowPadding,Left:=SuperLowPadding}
                                                HorizontalAlignment:=horizontal_alignment.Center
                                                VerticalAlignment:=vertical_alignment.Bottom
                                                Widget:=stack_box:  
                                                    Orientation:=orientation.Vertical
                                                        Slots:=array:
                                                            stack_box_slot: 
                                                                Padding:=margin{Top:=SelectorBoxPadding,Bottom:=SelectorBoxPadding}
                                                                Widget:=overlay:
                                                                    Slots:=array:
                                                                        overlay_slot:
                                                                            HorizontalAlignment:=horizontal_alignment.Fill
                                                                            VerticalAlignment:=vertical_alignment.Fill
                                                                            Widget:=texture_block{DefaultImage:=Images.Img_LineBackground}
                                                                        overlay_slot:
                                                                            HorizontalAlignment:=horizontal_alignment.Fill
                                                                            VerticalAlignment:=vertical_alignment.Fill
                                                                            Padding:=margin{Left:=SelectorBoxPadding,Right:=SelectorBoxPadding,Top:=SelectorBoxPadding,Bottom:=SelectorBoxPadding}
                                                                            Widget:=texture_block{DefaultImage:=Images.Img_LineBackground}
                                                                        overlay_slot:
                                                                            HorizontalAlignment:=horizontal_alignment.Center
                                                                            VerticalAlignment:=vertical_alignment.Center
                                                                            Widget:=texture_block{DefaultImage:=Images.EmptyTexture,DefaultDesiredSize:=vector2{X:=258.0,Y:=300.0}}
                                                                        overlay_slot:
                                                                            HorizontalAlignment:=horizontal_alignment.Center
                                                                            VerticalAlignment:=vertical_alignment.Center
                                                                            Widget:=stack_box:
                                                                                Orientation:=orientation.Vertical
                                                                                Slots:=array:
                                                                                    stack_box_slot: 
                                                                                        Padding:=RowPadding
                                                                                        HorizontalAlignment:=horizontal_alignment.Center
                                                                                        VerticalAlignment:=vertical_alignment.Center
                                                                                        Widget:=TextureBlock_PetImages_Box1
                                                                                    stack_box_slot: 
                                                                                        HorizontalAlignment:=horizontal_alignment.Center
                                                                                        VerticalAlignment:=vertical_alignment.Center
                                                                                        #Padding:=margin{Left:=TextPetsBoxPadding}
                                                                                        Widget:=stack_box:
                                                                                            Orientation:=orientation.Horizontal
                                                                                            Slots:=array:
                                                                                                stack_box_slot: 
                                                                                                    Padding:=margin{Left:=2.5,Right:=2.5}
                                                                                                    Widget:=TextBlock_Classes_Box1
                                                                                                stack_box_slot: 
                                                                                                    Padding:=margin{Left:=2.5, Right:=1.0}
                                                                                                    Widget:=TextBlock_PackA_Box1
                                                                                                stack_box_slot: 
                                                                                                    Padding:=margin{Right:=2.5}
                                                                                                    Widget:=TextBlock_PackB_Box1
                                                                                    stack_box_slot: 
                                                                                        Padding:=RowPadding
                                                                                        HorizontalAlignment:=horizontal_alignment.Center
                                                                                        VerticalAlignment:=vertical_alignment.Center
                                                                                        Widget:=TextBlock_Purity_Box1
                                                                                    stack_box_slot: 
                                                                                        Padding:=RowPadding
                                                                                        HorizontalAlignment:=horizontal_alignment.Center
                                                                                        VerticalAlignment:=vertical_alignment.Center
                                                                                        Widget:=TextBlock_Tier_Box1
                                                            stack_box_slot: 
                                                                Widget:=overlay:
                                                                    Slots:=array:
                                                                        overlay_slot:
                                                                            HorizontalAlignment:=horizontal_alignment.Fill
                                                                            VerticalAlignment:=vertical_alignment.Fill
                                                                            Widget:=texture_block{DefaultImage:=Images.EmptyTexture,DefaultDesiredSize:=QuietButton_PetSelectorDesiredSize}

                                            stack_box_slot: #SELECTED
                                                Padding:=margin{Right:=SuperLowPadding,Left:=SuperLowPadding}
                                                HorizontalAlignment:=horizontal_alignment.Center
                                                VerticalAlignment:=vertical_alignment.Bottom
                                                Widget:=stack_box:  
                                                    Orientation:=orientation.Vertical
                                                    Slots:=array:
                                                        stack_box_slot: 
                                                            Padding:=margin{Top:=SelectorBoxPadding,Bottom:=SelectorBoxPadding}
                                                            Widget:=overlay:
                                                                Slots:=array:
                                                                    overlay_slot:
                                                                        HorizontalAlignment:=horizontal_alignment.Fill
                                                                        VerticalAlignment:=vertical_alignment.Fill
                                                                        Widget:=texture_block{DefaultImage:=Images.Img_LineBackground}
                                                                    overlay_slot:
                                                                        HorizontalAlignment:=horizontal_alignment.Fill
                                                                        VerticalAlignment:=vertical_alignment.Fill
                                                                        Padding:=margin{Left:=SelectorBoxPadding,Right:=SelectorBoxPadding,Top:=SelectorBoxPadding,Bottom:=SelectorBoxPadding}
                                                                        Widget:=texture_block{DefaultImage:=Images.Img_LineBackground}
                                                                    overlay_slot:
                                                                        HorizontalAlignment:=horizontal_alignment.Center
                                                                        VerticalAlignment:=vertical_alignment.Center
                                                                        Widget:=texture_block{DefaultImage:=Images.EmptyTexture,DefaultDesiredSize:=vector2{X:=1300.0,Y:=705.0}}
                                                                    overlay_slot:
                                                                        HorizontalAlignment:=horizontal_alignment.Center
                                                                        VerticalAlignment:=vertical_alignment.Center
                                                                        Widget:=stack_box:  
                                                                            Orientation:=orientation.Vertical
                                                                            Slots:=array:
                                                                                stack_box_slot: 
                                                                                    Widget:=stack_box:  
                                                                                        Orientation:=orientation.Horizontal
                                                                                        Slots:=array:
                                                                                            stack_box_slot: 
                                                                                                Padding:=RowPadding
                                                                                                Widget:=Selected_TextureBlock_PetImages
                                                                                            stack_box_slot: 
                                                                                                HorizontalAlignment:=horizontal_alignment.Center
                                                                                                VerticalAlignment:=vertical_alignment.Center
                                                                                                Widget:=stack_box:  
                                                                                                    Orientation:=orientation.Vertical
                                                                                                    Slots:=array:
                                                                                                        stack_box_slot: 
                                                                                                            HorizontalAlignment:=horizontal_alignment.Center
                                                                                                            VerticalAlignment:=vertical_alignment.Center
                                                                                                            #Padding:=margin{Left:=TextPetsBoxPadding}
                                                                                                            Widget:=stack_box:
                                                                                                                Orientation:=orientation.Horizontal
                                                                                                                Slots:=array:
                                                                                                                    stack_box_slot: 
                                                                                                                        Padding:=margin{Left:=2.5,Right:=2.5}
                                                                                                                        Widget:=Selected_TextBlock_Class
                                                                                                                    stack_box_slot: 
                                                                                                                        Padding:=margin{Left:=2.5, Right:=1.0}
                                                                                                                        Widget:=Selected_TextBlock_PackA
                                                                                                                    stack_box_slot: 
                                                                                                                        Padding:=margin{Right:=2.5}
                                                                                                                        Widget:=Selected_TextBlock_PackB
                                                                                                        stack_box_slot: 
                                                                                                            Padding:=RowPadding
                                                                                                            Widget:=Selected_TextBlock_Purity
                                                                                stack_box_slot: 
                                                                                    HorizontalAlignment:=horizontal_alignment.Fill
                                                                                    VerticalAlignment:=vertical_alignment.Fill
                                                                                    Widget:=texture_block{DefaultImage:=Images.Img_HorizontalLine,DefaultDesiredSize:=vector2{Y:=3.0}}
                                                                                stack_box_slot:  
                                                                                    HorizontalAlignment:=horizontal_alignment.Center
                                                                                    VerticalAlignment:=vertical_alignment.Center
                                                                                    Padding:=RowPadding
                                                                                    Widget:=Text_PackageA_LV
                                                                                stack_box_slot: 
                                                                                    HorizontalAlignment:=horizontal_alignment.Fill
                                                                                    VerticalAlignment:=vertical_alignment.Fill
                                                                                    Widget:=texture_block{DefaultImage:=Images.Img_HorizontalLine,DefaultDesiredSize:=vector2{Y:=3.0}}
                                                                                stack_box_slot: 
                                                                                    HorizontalAlignment:=horizontal_alignment.Fill
                                                                                    VerticalAlignment:=vertical_alignment.Fill
                                                                                    Widget:=stack_box:  
                                                                                        Orientation:=orientation.Horizontal
                                                                                        Slots:=array:
                                                                                            stack_box_slot: 
                                                                                                Widget:=stack_box:
                                                                                                    Orientation:=orientation.Vertical
                                                                                                    Slots:=array:
                                                                                                        stack_box_slot: 
                                                                                                            Padding:=RowPadding
                                                                                                            Widget:=text_block{DefaultText:=BasicMessage("Tournament stats"), DefaultTextColor:=Red, DefaultShadowOffset:=ShadowOffset, DefaultShadowOpacity := DefaultTextShadowOpacity}
                                                                                                        stack_box_slot: 
                                                                                                            HorizontalAlignment:=horizontal_alignment.Fill
                                                                                                            VerticalAlignment:=vertical_alignment.Fill
                                                                                                            Widget:=stack_box:  
                                                                                                                Orientation:=orientation.Horizontal
                                                                                                                Slots:=array:
                                                                                                                    stack_box_slot: 
                                                                                                                        HorizontalAlignment:=horizontal_alignment.Fill
                                                                                                                        VerticalAlignment:=vertical_alignment.Fill
                                                                                                                        Widget:=stack_box:  
                                                                                                                            Orientation:=orientation.Vertical
                                                                                                                            Slots:=array:
                                                                                                                                stack_box_slot: 
                                                                                                                                    Widget:=overlay:
                                                                                                                                        Slots:=array:
                                                                                                                                            overlay_slot:
                                                                                                                                                Widget:=text_block{DefaultText:=BasicMessage("")}
                                                                                                                                            overlay_slot:
                                                                                                                                                Widget:=texture_block{DefaultImage:=Images.EmptyTexture,DefaultDesiredSize:=vector2{X:=230.0,Y:=32.0}}
                                                                                                                                stack_box_slot: 
                                                                                                                                    HorizontalAlignment:=horizontal_alignment.Left
                                                                                                                                    Padding:=RowPadding
                                                                                                                                    Widget:=text_block{DefaultText:=BasicMessage("Health:"), DefaultTextColor:=White, DefaultShadowOffset:=ShadowOffset, DefaultShadowOpacity := DefaultTextShadowOpacity}
                                                                                                                                stack_box_slot: 
                                                                                                                                    HorizontalAlignment:=horizontal_alignment.Left
                                                                                                                                    Padding:=RowPadding
                                                                                                                                    Widget:=text_block{DefaultText:=BasicMessage("Damage:"), DefaultTextColor:=White, DefaultShadowOffset:=ShadowOffset, DefaultShadowOpacity := DefaultTextShadowOpacity}
                                                                                                                                stack_box_slot: 
                                                                                                                                    HorizontalAlignment:=horizontal_alignment.Left
                                                                                                                                    Padding:=RowPadding
                                                                                                                                    Widget:=text_block{DefaultText:=BasicMessage("Mov. speed:"), DefaultTextColor:=White, DefaultShadowOffset:=ShadowOffset, DefaultShadowOpacity := DefaultTextShadowOpacity}
                                                                                                                    stack_box_slot: 
                                                                                                                        HorizontalAlignment:=horizontal_alignment.Fill
                                                                                                                        VerticalAlignment:=vertical_alignment.Fill
                                                                                                                        Widget:=stack_box:  
                                                                                                                            Orientation:=orientation.Vertical
                                                                                                                            Slots:=array:
                                                                                                                                stack_box_slot: 
                                                                                                                                    Widget:=overlay:
                                                                                                                                        Slots:=array:
                                                                                                                                            overlay_slot:
                                                                                                                                                Padding:=RowPadding
                                                                                                                                                Widget:=text_block{DefaultText:=BasicMessage("Current"), DefaultTextColor:=CurrentTextColor, DefaultShadowOffset:=ShadowOffset, DefaultShadowOpacity := DefaultTextShadowOpacity}
                                                                                                                                            overlay_slot:
                                                                                                                                                Widget:=texture_block{DefaultImage:=Images.EmptyTexture,DefaultDesiredSize:=vector2{X:=257.0,Y:=32.0}}
                                                                                                                                stack_box_slot: 
                                                                                                                                    Padding:=RowPadding
                                                                                                                                    Widget:=Selected_TextBlock_Current_TournamentHealth
                                                                                                                                stack_box_slot: 
                                                                                                                                    Padding:=RowPadding
                                                                                                                                    Widget:=Selected_TextBlock_Current_TournamentDamage
                                                                                                                                stack_box_slot: 
                                                                                                                                    Padding:=RowPadding
                                                                                                                                    Widget:=Selected_TextBlock_Current_TournamentSpeed
                                                                                                                    stack_box_slot: 
                                                                                                                        HorizontalAlignment:=horizontal_alignment.Fill
                                                                                                                        VerticalAlignment:=vertical_alignment.Fill
                                                                                                                        Widget:=stack_box:  
                                                                                                                            Orientation:=orientation.Vertical
                                                                                                                            Slots:=array:
                                                                                                                                stack_box_slot: 
                                                                                                                                    Widget:=overlay:
                                                                                                                                        Slots:=array:
                                                                                                                                            overlay_slot:
                                                                                                                                                Padding:=RowPadding
                                                                                                                                                Widget:=text_block{DefaultText:=BasicMessage("Upgrade"), DefaultTextColor:=UpgradeTextColor, DefaultShadowOffset:=ShadowOffset, DefaultShadowOpacity := DefaultTextShadowOpacity}
                                                                                                                                            overlay_slot:
                                                                                                                                                Widget:=texture_block{DefaultImage:=Images.EmptyTexture,DefaultDesiredSize:=vector2{X:=155.0,Y:=32.0}}
                                                                                                                                stack_box_slot: 
                                                                                                                                    Padding:=RowPadding
                                                                                                                                    Widget:=Selected_TextBlock_Upgrade_TournamentHealth
                                                                                                                                stack_box_slot: 
                                                                                                                                    Padding:=RowPadding
                                                                                                                                    Widget:=Selected_TextBlock_Upgrade_TournamentDamage
                                                                                                                                stack_box_slot: 
                                                                                                                                    Padding:=RowPadding
                                                                                                                                    Widget:=Selected_TextBlock_Upgrade_TournamentSpeed
                                                                                            stack_box_slot: 
                                                                                                HorizontalAlignment:=horizontal_alignment.Center
                                                                                                VerticalAlignment:=vertical_alignment.Fill
                                                                                                Widget:=texture_block{DefaultImage:=Images.Img_VerticalLine,DefaultDesiredSize:=vector2{X:=10.0}}
                                                                                            stack_box_slot: 
                                                                                                Widget:=stack_box:
                                                                                                    Orientation:=orientation.Vertical
                                                                                                    Slots:=array:
                                                                                                        stack_box_slot: 
                                                                                                            Padding:=RowPadding
                                                                                                            Widget:=text_block{DefaultText:=BasicMessage("Player enhancements"), DefaultTextColor:=Cyan, DefaultShadowOffset:=ShadowOffset, DefaultShadowOpacity := DefaultTextShadowOpacity}
                                                                                                        stack_box_slot: 
                                                                                                            HorizontalAlignment:=horizontal_alignment.Fill
                                                                                                            VerticalAlignment:=vertical_alignment.Fill
                                                                                                            Widget:=stack_box:  
                                                                                                                Orientation:=orientation.Horizontal
                                                                                                                Slots:=array:
                                                                                                                    stack_box_slot: 
                                                                                                                        HorizontalAlignment:=horizontal_alignment.Fill
                                                                                                                        VerticalAlignment:=vertical_alignment.Fill
                                                                                                                        Widget:=stack_box:  
                                                                                                                            Orientation:=orientation.Vertical
                                                                                                                            Slots:=array:
                                                                                                                                stack_box_slot: 
                                                                                                                                    Widget:=overlay:
                                                                                                                                        Slots:=array:
                                                                                                                                            overlay_slot:
                                                                                                                                                Widget:=text_block{DefaultText:=BasicMessage("")}
                                                                                                                                            overlay_slot:
                                                                                                                                                Widget:=texture_block{DefaultImage:=Images.EmptyTexture,DefaultDesiredSize:=vector2{X:=300.0,Y:=32.0}}
                                                                                                                                stack_box_slot: 
                                                                                                                                    Padding:=RowPadding
                                                                                                                                    HorizontalAlignment:=horizontal_alignment.Left
                                                                                                                                    Widget:=Selected_TextBlock_Description_SpecialPerkA
                                                                                                                                stack_box_slot: 
                                                                                                                                    Padding:=RowPadding
                                                                                                                                    HorizontalAlignment:=horizontal_alignment.Left
                                                                                                                                    Widget:=text_block{DefaultText:=BasicMessage("Pick damage:"), DefaultTextColor:=White, DefaultShadowOffset:=ShadowOffset, DefaultShadowOpacity := DefaultTextShadowOpacity}
                                                                                                                                stack_box_slot: 
                                                                                                                                    Padding:=RowPadding
                                                                                                                                    HorizontalAlignment:=horizontal_alignment.Left
                                                                                                                                    Widget:=text_block{DefaultText:=BasicMessage("Weak point mult:"), DefaultTextColor:=White, DefaultShadowOffset:=ShadowOffset, DefaultShadowOpacity := DefaultTextShadowOpacity}
                                                                                                                    stack_box_slot: 
                                                                                                                        HorizontalAlignment:=horizontal_alignment.Fill
                                                                                                                        VerticalAlignment:=vertical_alignment.Fill
                                                                                                                        Widget:=stack_box:  
                                                                                                                            Orientation:=orientation.Vertical
                                                                                                                            Slots:=array:
                                                                                                                                stack_box_slot: 
                                                                                                                                    Widget:=overlay:
                                                                                                                                        Slots:=array:
                                                                                                                                            overlay_slot:
                                                                                                                                                Padding:=RowPadding
                                                                                                                                                Widget:=text_block{DefaultText:=BasicMessage("Current"), DefaultTextColor:=CurrentTextColor, DefaultShadowOffset:=ShadowOffset, DefaultShadowOpacity := DefaultTextShadowOpacity}
                                                                                                                                            overlay_slot:
                                                                                                                                                Widget:=texture_block{DefaultImage:=Images.EmptyTexture,DefaultDesiredSize:=vector2{X:=180.0,Y:=32.0}}
                                                                                                                                stack_box_slot: 
                                                                                                                                    Padding:=RowPadding
                                                                                                                                    Widget:=Selected_TextBlock_Current_SpecialPerkA
                                                                                                                                stack_box_slot: 
                                                                                                                                    Padding:=RowPadding
                                                                                                                                    Widget:=Selected_TextBlock_Current_PlayerPickDamage
                                                                                                                                stack_box_slot:
                                                                                                                                    Padding:=RowPadding 
                                                                                                                                    Widget:=Selected_TextBlock_Current_PlayerWeakPoint
                                                                                                                    stack_box_slot: 
                                                                                                                        HorizontalAlignment:=horizontal_alignment.Fill
                                                                                                                        VerticalAlignment:=vertical_alignment.Fill
                                                                                                                        Widget:=stack_box:  
                                                                                                                            Orientation:=orientation.Vertical
                                                                                                                            Slots:=array:
                                                                                                                                stack_box_slot: 
                                                                                                                                    Widget:=overlay:
                                                                                                                                        Slots:=array:
                                                                                                                                            overlay_slot:
                                                                                                                                                Padding:=RowPadding
                                                                                                                                                Widget:=text_block{DefaultText:=BasicMessage("Upgrade"), DefaultTextColor:=UpgradeTextColor, DefaultShadowOffset:=ShadowOffset, DefaultShadowOpacity := DefaultTextShadowOpacity}
                                                                                                                                            overlay_slot:
                                                                                                                                                Widget:=texture_block{DefaultImage:=Images.EmptyTexture,DefaultDesiredSize:=vector2{X:=155.0,Y:=32.0}}
                                                                                                                                stack_box_slot: 
                                                                                                                                    Padding:=RowPadding
                                                                                                                                    Widget:=Selected_TextBlock_Upgrade_SpecialPerkA
                                                                                                                                stack_box_slot: 
                                                                                                                                    Padding:=RowPadding
                                                                                                                                    Widget:=Selected_TextBlock_Upgrade_PlayerPickDamage
                                                                                                                                stack_box_slot: 
                                                                                                                                    Padding:=RowPadding
                                                                                                                                    Widget:=Selected_TextBlock_Upgrade_PlayerWeakPoint

                                                                                stack_box_slot: 
                                                                                    HorizontalAlignment:=horizontal_alignment.Fill
                                                                                    VerticalAlignment:=vertical_alignment.Fill
                                                                                    Widget:=texture_block{DefaultImage:=Images.Img_HorizontalLine,DefaultDesiredSize:=vector2{Y:=3.0}}
                                                                                stack_box_slot:  
                                                                                    HorizontalAlignment:=horizontal_alignment.Center
                                                                                    VerticalAlignment:=vertical_alignment.Center
                                                                                    Padding:=RowPadding
                                                                                    Widget:=Text_PackageB_LV
                                                                                stack_box_slot: 
                                                                                    HorizontalAlignment:=horizontal_alignment.Fill
                                                                                    VerticalAlignment:=vertical_alignment.Fill
                                                                                    Widget:=texture_block{DefaultImage:=Images.Img_HorizontalLine,DefaultDesiredSize:=vector2{Y:=3.0}}
                                                                                stack_box_slot: 
                                                                                    HorizontalAlignment:=horizontal_alignment.Fill
                                                                                    VerticalAlignment:=vertical_alignment.Fill
                                                                                    Widget:=stack_box:  
                                                                                        Orientation:=orientation.Horizontal
                                                                                        Slots:=array:
                                                                                            stack_box_slot: 
                                                                                                Widget:=stack_box:
                                                                                                    Orientation:=orientation.Vertical
                                                                                                    Slots:=array:
                                                                                                        stack_box_slot: 
                                                                                                            Padding:=RowPadding
                                                                                                            Widget:=text_block{DefaultText:=BasicMessage("Tournament stats"), DefaultTextColor:=Red, DefaultShadowOffset:=ShadowOffset, DefaultShadowOpacity := DefaultTextShadowOpacity}
                                                                                                        stack_box_slot: 
                                                                                                            HorizontalAlignment:=horizontal_alignment.Fill
                                                                                                            VerticalAlignment:=vertical_alignment.Fill
                                                                                                            Widget:=stack_box:  
                                                                                                                Orientation:=orientation.Horizontal
                                                                                                                Slots:=array:
                                                                                                                    stack_box_slot: 
                                                                                                                        HorizontalAlignment:=horizontal_alignment.Fill
                                                                                                                        VerticalAlignment:=vertical_alignment.Fill
                                                                                                                        Widget:=stack_box:  
                                                                                                                            Orientation:=orientation.Vertical
                                                                                                                            Slots:=array:
                                                                                                                                stack_box_slot: 
                                                                                                                                    Widget:=overlay:
                                                                                                                                        Slots:=array:
                                                                                                                                            overlay_slot:
                                                                                                                                                Widget:=text_block{DefaultText:=BasicMessage("")}
                                                                                                                                            overlay_slot:
                                                                                                                                                Widget:=texture_block{DefaultImage:=Images.EmptyTexture,DefaultDesiredSize:=vector2{X:=230.0,Y:=32.0}}
                                                                                                                                stack_box_slot: 
                                                                                                                                    Padding:=RowPadding
                                                                                                                                    HorizontalAlignment:=horizontal_alignment.Left
                                                                                                                                    Widget:=text_block{DefaultText:=BasicMessage("Attack time:"), DefaultTextColor:=White, DefaultShadowOffset:=ShadowOffset, DefaultShadowOpacity := DefaultTextShadowOpacity}
                                                                                                                                stack_box_slot: 
                                                                                                                                    Padding:=RowPadding
                                                                                                                                    HorizontalAlignment:=horizontal_alignment.Left
                                                                                                                                    Widget:=text_block{DefaultText:=BasicMessage("Range:"), DefaultTextColor:=White, DefaultShadowOffset:=ShadowOffset, DefaultShadowOpacity := DefaultTextShadowOpacity}
                                                                                                                                stack_box_slot: 
                                                                                                                                    Padding:=RowPadding
                                                                                                                                    HorizontalAlignment:=horizontal_alignment.Left
                                                                                                                                    Widget:=text_block{DefaultText:=BasicMessage("Healing perk:"), DefaultTextColor:=White, DefaultShadowOffset:=ShadowOffset, DefaultShadowOpacity := DefaultTextShadowOpacity}
                                                                                                                    stack_box_slot: 
                                                                                                                        HorizontalAlignment:=horizontal_alignment.Fill
                                                                                                                        VerticalAlignment:=vertical_alignment.Fill
                                                                                                                        Widget:=stack_box:  
                                                                                                                            Orientation:=orientation.Vertical
                                                                                                                            Slots:=array:
                                                                                                                                stack_box_slot: 
                                                                                                                                    Widget:=overlay:
                                                                                                                                        Slots:=array:
                                                                                                                                            overlay_slot:
                                                                                                                                                Padding:=RowPadding
                                                                                                                                                Widget:=text_block{DefaultText:=BasicMessage("Current"), DefaultTextColor:=CurrentTextColor, DefaultShadowOffset:=ShadowOffset, DefaultShadowOpacity := DefaultTextShadowOpacity}
                                                                                                                                            overlay_slot:
                                                                                                                                                Widget:=texture_block{DefaultImage:=Images.EmptyTexture,DefaultDesiredSize:=vector2{X:=257.0,Y:=32.0}}
                                                                                                                                stack_box_slot: 
                                                                                                                                    Padding:=RowPadding
                                                                                                                                    Widget:=Selected_TextBlock_Current_TournamentActionSpeed
                                                                                                                                stack_box_slot: 
                                                                                                                                    Padding:=RowPadding
                                                                                                                                    Widget:=Selected_TextBlock_Current_TournamentRange
                                                                                                                                stack_box_slot: 
                                                                                                                                    Padding:=RowPadding
                                                                                                                                    Widget:=Selected_TextBlock_Current_TournamentHealingPerk
                                                                                                                    stack_box_slot: 
                                                                                                                        HorizontalAlignment:=horizontal_alignment.Fill
                                                                                                                        VerticalAlignment:=vertical_alignment.Fill
                                                                                                                        Widget:=stack_box:  
                                                                                                                            Orientation:=orientation.Vertical
                                                                                                                            Slots:=array:
                                                                                                                                stack_box_slot: 
                                                                                                                                    Widget:=overlay:
                                                                                                                                        Slots:=array:
                                                                                                                                            overlay_slot:
                                                                                                                                                Padding:=RowPadding
                                                                                                                                                Widget:=text_block{DefaultText:=BasicMessage("Upgrade"), DefaultTextColor:=UpgradeTextColor, DefaultShadowOffset:=ShadowOffset, DefaultShadowOpacity := DefaultTextShadowOpacity}
                                                                                                                                            overlay_slot:
                                                                                                                                                Widget:=texture_block{DefaultImage:=Images.EmptyTexture,DefaultDesiredSize:=vector2{X:=155.0,Y:=32.0}}
                                                                                                                                stack_box_slot: 
                                                                                                                                    Padding:=RowPadding
                                                                                                                                    Widget:=Selected_TextBlock_Upgrade_TournamentActionSpeed
                                                                                                                                stack_box_slot: 
                                                                                                                                    Padding:=RowPadding
                                                                                                                                    Widget:=Selected_TextBlock_Upgrade_TournamentRange
                                                                                                                                stack_box_slot: 
                                                                                                                                    Padding:=RowPadding
                                                                                                                                    Widget:=Selected_TextBlock_Upgrade_TournamentHealingPerk
                                                                                            stack_box_slot: 
                                                                                                HorizontalAlignment:=horizontal_alignment.Center
                                                                                                VerticalAlignment:=vertical_alignment.Fill
                                                                                                Widget:=texture_block{DefaultImage:=Images.Img_VerticalLine,DefaultDesiredSize:=vector2{X:=10.0}}
                                                                                            stack_box_slot: 
                                                                                                Widget:=stack_box:
                                                                                                    Orientation:=orientation.Vertical
                                                                                                    Slots:=array:
                                                                                                        stack_box_slot: 
                                                                                                            Padding:=RowPadding
                                                                                                            Widget:=text_block{DefaultText:=BasicMessage("Player enhancements"), DefaultTextColor:=Cyan, DefaultShadowOffset:=ShadowOffset, DefaultShadowOpacity := DefaultTextShadowOpacity}
                                                                                                        stack_box_slot: 
                                                                                                            HorizontalAlignment:=horizontal_alignment.Fill
                                                                                                            VerticalAlignment:=vertical_alignment.Fill
                                                                                                            Widget:=stack_box:  
                                                                                                                Orientation:=orientation.Horizontal
                                                                                                                Slots:=array:
                                                                                                                    stack_box_slot: 
                                                                                                                        HorizontalAlignment:=horizontal_alignment.Fill
                                                                                                                        VerticalAlignment:=vertical_alignment.Fill
                                                                                                                        Widget:=stack_box:  
                                                                                                                            Orientation:=orientation.Vertical
                                                                                                                            Slots:=array:
                                                                                                                                stack_box_slot: 
                                                                                                                                    Widget:=overlay:
                                                                                                                                        Slots:=array:
                                                                                                                                            overlay_slot:
                                                                                                                                                Widget:=text_block{DefaultText:=BasicMessage("")}
                                                                                                                                            overlay_slot:
                                                                                                                                                Widget:=texture_block{DefaultImage:=Images.EmptyTexture,DefaultDesiredSize:=vector2{X:=300.0,Y:=32.0}}
                                                                                                                                stack_box_slot: 
                                                                                                                                    Padding:=RowPadding
                                                                                                                                    HorizontalAlignment:=horizontal_alignment.Left
                                                                                                                                    Widget:=Selected_TextBlock_Description_SpecialPerkB
                                                                                                                                stack_box_slot: 
                                                                                                                                    Padding:=RowPadding
                                                                                                                                    HorizontalAlignment:=horizontal_alignment.Left
                                                                                                                                    Widget:=text_block{DefaultText:=BasicMessage("Pick efficiency:"), DefaultTextColor:=White, DefaultShadowOffset:=ShadowOffset, DefaultShadowOpacity := DefaultTextShadowOpacity}
                                                                                                                                stack_box_slot: 
                                                                                                                                    Padding:=RowPadding
                                                                                                                                    HorizontalAlignment:=horizontal_alignment.Left
                                                                                                                                    Widget:=text_block{DefaultText:=BasicMessage("Pick critical dam:"), DefaultTextColor:=White, DefaultShadowOffset:=ShadowOffset, DefaultShadowOpacity := DefaultTextShadowOpacity}
                                                                                                                    stack_box_slot: 
                                                                                                                        HorizontalAlignment:=horizontal_alignment.Fill
                                                                                                                        VerticalAlignment:=vertical_alignment.Fill
                                                                                                                        Widget:=stack_box:  
                                                                                                                            Orientation:=orientation.Vertical
                                                                                                                            Slots:=array:
                                                                                                                                stack_box_slot: 
                                                                                                                                    Widget:=overlay:
                                                                                                                                        Slots:=array:
                                                                                                                                            overlay_slot:
                                                                                                                                                Padding:=RowPadding
                                                                                                                                                Widget:=text_block{DefaultText:=BasicMessage("Current"), DefaultTextColor:=CurrentTextColor, DefaultShadowOffset:=ShadowOffset, DefaultShadowOpacity := DefaultTextShadowOpacity}
                                                                                                                                            overlay_slot:
                                                                                                                                                Widget:=texture_block{DefaultImage:=Images.EmptyTexture,DefaultDesiredSize:=vector2{X:=180.0,Y:=32.0}}
                                                                                                                                stack_box_slot: 
                                                                                                                                    Padding:=RowPadding
                                                                                                                                    Widget:=Selected_TextBlock_Current_SpecialPerkB
                                                                                                                                stack_box_slot: 
                                                                                                                                    Padding:=RowPadding
                                                                                                                                    Widget:=Selected_TextBlock_Current_PlayerPickEfficiency
                                                                                                                                stack_box_slot: 
                                                                                                                                    Padding:=RowPadding
                                                                                                                                    Widget:=Selected_TextBlock_Current_PlayerPickCriticalDamage
                                                                                                                    stack_box_slot: 
                                                                                                                        HorizontalAlignment:=horizontal_alignment.Fill
                                                                                                                        VerticalAlignment:=vertical_alignment.Fill
                                                                                                                        Widget:=stack_box:  
                                                                                                                            Orientation:=orientation.Vertical
                                                                                                                            Slots:=array:
                                                                                                                                stack_box_slot: 
                                                                                                                                    Widget:=overlay:
                                                                                                                                        Slots:=array:
                                                                                                                                            overlay_slot:
                                                                                                                                                Padding:=RowPadding
                                                                                                                                                Widget:=text_block{DefaultText:=BasicMessage("Upgrade"), DefaultTextColor:=UpgradeTextColor, DefaultShadowOffset:=ShadowOffset, DefaultShadowOpacity := DefaultTextShadowOpacity}
                                                                                                                                            overlay_slot:
                                                                                                                                                Widget:=texture_block{DefaultImage:=Images.EmptyTexture,DefaultDesiredSize:=vector2{X:=155.0,Y:=32.0}}
                                                                                                                                stack_box_slot: 
                                                                                                                                    Padding:=RowPadding
                                                                                                                                    Widget:=Selected_TextBlock_Upgrade_SpecialPerkB
                                                                                                                                stack_box_slot: 
                                                                                                                                    Padding:=RowPadding
                                                                                                                                    Widget:=Selected_TextBlock_Upgrade_PlayerPickEfficiency
                                                                                                                                stack_box_slot: 
                                                                                                                                    Padding:=RowPadding
                                                                                                                                    Widget:=Selected_TextBlock_Upgrade_PlayerPickCriticalDamage
                                                        stack_box_slot: #Buttons Here 
                                                            Padding:=margin{Top:=SelectorBoxPadding,Bottom:=SelectorBoxPadding}
                                                            Widget:=overlay:
                                                                Slots:=array:
                                                                    overlay_slot:
                                                                        Widget:=stack_box:  
                                                                            Orientation:=orientation.Horizontal
                                                                                Slots:=array:
                                                                                    stack_box_slot: 
                                                                                        Padding:=RowPadding 
                                                                                        Widget:=overlay:
                                                                                            Slots:=array:
                                                                                                overlay_slot:
                                                                                                    Widget:=texture_block{DefaultImage:=Images.EmptyTexture,DefaultDesiredSize:=vector2{X:=530.0,Y:=32.0}}
                                                                                                overlay_slot:
                                                                                                    HorizontalAlignment:=horizontal_alignment.Fill
                                                                                                    Widget:=NavigateLeftButton
                                                                                    stack_box_slot: 
                                                                                        Padding:=RowPadding 
                                                                                        Widget:=overlay:
                                                                                            Slots:=array:
                                                                                                overlay_slot:
                                                                                                    Widget:=texture_block{DefaultImage:=Images.EmptyTexture,DefaultDesiredSize:=vector2{X:=530.0,Y:=32.0}}
                                                                                                overlay_slot:
                                                                                                    HorizontalAlignment:=horizontal_alignment.Fill
                                                                                                    Widget:=NavigateRightButton


                                            stack_box_slot: #Slot 2
                                                Padding:=margin{Right:=SuperLowPadding,Left:=SuperLowPadding}
                                                HorizontalAlignment:=horizontal_alignment.Center
                                                VerticalAlignment:=vertical_alignment.Bottom
                                                Widget:=stack_box:  
                                                    Orientation:=orientation.Vertical
                                                        Slots:=array:
                                                            stack_box_slot: 
                                                                Padding:=margin{Top:=SelectorBoxPadding,Bottom:=SelectorBoxPadding}
                                                                Widget:=overlay:
                                                                    Slots:=array:
                                                                        overlay_slot:
                                                                            HorizontalAlignment:=horizontal_alignment.Fill
                                                                            VerticalAlignment:=vertical_alignment.Fill
                                                                            Widget:=texture_block{DefaultImage:=Images.Img_LineBackground}
                                                                        overlay_slot:
                                                                            HorizontalAlignment:=horizontal_alignment.Fill
                                                                            VerticalAlignment:=vertical_alignment.Fill
                                                                            Padding:=margin{Left:=SelectorBoxPadding,Right:=SelectorBoxPadding,Top:=SelectorBoxPadding,Bottom:=SelectorBoxPadding}
                                                                            Widget:=texture_block{DefaultImage:=Images.Img_LineBackground}
                                                                        overlay_slot:
                                                                            HorizontalAlignment:=horizontal_alignment.Center
                                                                            VerticalAlignment:=vertical_alignment.Center
                                                                            Widget:=texture_block{DefaultImage:=Images.EmptyTexture,DefaultDesiredSize:=vector2{X:=258.0,Y:=300.0}}
                                                                        overlay_slot:
                                                                            HorizontalAlignment:=horizontal_alignment.Center
                                                                            VerticalAlignment:=vertical_alignment.Center
                                                                            Widget:=stack_box:
                                                                                Orientation:=orientation.Vertical
                                                                                Slots:=array:
                                                                                    stack_box_slot: 
                                                                                        Padding:=RowPadding
                                                                                        HorizontalAlignment:=horizontal_alignment.Center
                                                                                        VerticalAlignment:=vertical_alignment.Center
                                                                                        Widget:=TextureBlock_PetImages_Box2
                                                                                    stack_box_slot: 
                                                                                        HorizontalAlignment:=horizontal_alignment.Left
                                                                                        VerticalAlignment:=vertical_alignment.Center
                                                                                        #Padding:=margin{Left:=TextPetsBoxPadding}
                                                                                        Widget:=stack_box:
                                                                                            Orientation:=orientation.Horizontal
                                                                                            Slots:=array:
                                                                                                stack_box_slot: 
                                                                                                    Padding:=margin{Left:=2.5,Right:=2.5}
                                                                                                    Widget:=TextBlock_Classes_Box2
                                                                                                stack_box_slot: 
                                                                                                    Padding:=margin{Left:=2.5, Right:=1.0}
                                                                                                    Widget:=TextBlock_PackA_Box2
                                                                                                stack_box_slot: 
                                                                                                    Padding:=margin{Right:=2.5}
                                                                                                    Widget:=TextBlock_PackB_Box2
                                                                                    stack_box_slot: 
                                                                                        Padding:=RowPadding
                                                                                        HorizontalAlignment:=horizontal_alignment.Center
                                                                                        VerticalAlignment:=vertical_alignment.Center
                                                                                        Widget:=TextBlock_Purity_Box2
                                                                                    stack_box_slot: 
                                                                                        Padding:=RowPadding
                                                                                        HorizontalAlignment:=horizontal_alignment.Center
                                                                                        VerticalAlignment:=vertical_alignment.Center
                                                                                        Widget:=TextBlock_Tier_Box2
                                                            stack_box_slot: 
                                                                Widget:=overlay:
                                                                    Slots:=array:
                                                                        overlay_slot:
                                                                            HorizontalAlignment:=horizontal_alignment.Fill
                                                                            VerticalAlignment:=vertical_alignment.Fill
                                                                            Widget:=texture_block{DefaultImage:=Images.EmptyTexture,DefaultDesiredSize:=QuietButton_PetSelectorDesiredSize}

                                            stack_box_slot: #Slot 4
                                                Padding:=margin{Right:=SuperLowPadding,Left:=SuperLowPadding}
                                                HorizontalAlignment:=horizontal_alignment.Center
                                                VerticalAlignment:=vertical_alignment.Bottom
                                                Widget:=stack_box:  
                                                    Orientation:=orientation.Vertical
                                                        Slots:=array:
                                                            stack_box_slot: 
                                                                Padding:=margin{Top:=SelectorBoxPadding,Bottom:=SelectorBoxPadding}
                                                                Widget:=overlay:
                                                                    Slots:=array:
                                                                        overlay_slot:
                                                                            HorizontalAlignment:=horizontal_alignment.Fill
                                                                            VerticalAlignment:=vertical_alignment.Fill
                                                                            Widget:=texture_block{DefaultImage:=Images.Img_LineBackground}
                                                                        overlay_slot:
                                                                            HorizontalAlignment:=horizontal_alignment.Fill
                                                                            VerticalAlignment:=vertical_alignment.Fill
                                                                            Padding:=margin{Left:=SelectorBoxPadding,Right:=SelectorBoxPadding,Top:=SelectorBoxPadding,Bottom:=SelectorBoxPadding}
                                                                            Widget:=texture_block{DefaultImage:=Images.Img_LineBackground}
                                                                        overlay_slot:
                                                                            HorizontalAlignment:=horizontal_alignment.Center
                                                                            VerticalAlignment:=vertical_alignment.Center
                                                                            Widget:=texture_block{DefaultImage:=Images.EmptyTexture,DefaultDesiredSize:=vector2{X:=258.0,Y:=300.0}}
                                                                        overlay_slot:
                                                                            HorizontalAlignment:=horizontal_alignment.Center
                                                                            VerticalAlignment:=vertical_alignment.Center
                                                                            Widget:=stack_box:
                                                                                Orientation:=orientation.Vertical
                                                                                Slots:=array:
                                                                                    stack_box_slot: 
                                                                                        Padding:=RowPadding
                                                                                        HorizontalAlignment:=horizontal_alignment.Center
                                                                                        VerticalAlignment:=vertical_alignment.Center
                                                                                        Widget:=TextureBlock_PetImages_Box3
                                                                                    stack_box_slot: 
                                                                                        HorizontalAlignment:=horizontal_alignment.Left
                                                                                        VerticalAlignment:=vertical_alignment.Center
                                                                                        #Padding:=margin{Left:=TextPetsBoxPadding}
                                                                                        Widget:=stack_box:
                                                                                            Orientation:=orientation.Horizontal
                                                                                            Slots:=array:
                                                                                                stack_box_slot: 
                                                                                                    Padding:=margin{Left:=2.5,Right:=2.5}
                                                                                                    Widget:=TextBlock_Classes_Box3
                                                                                                stack_box_slot: 
                                                                                                    Padding:=margin{Left:=2.5, Right:=1.0}
                                                                                                    Widget:=TextBlock_PackA_Box3
                                                                                                stack_box_slot: 
                                                                                                    Padding:=margin{Right:=2.5}
                                                                                                    Widget:=TextBlock_PackB_Box3
                                                                                    stack_box_slot: 
                                                                                        Padding:=RowPadding
                                                                                        HorizontalAlignment:=horizontal_alignment.Center
                                                                                        VerticalAlignment:=vertical_alignment.Center
                                                                                        Widget:=TextBlock_Purity_Box3
                                                                                    stack_box_slot: 
                                                                                        Padding:=RowPadding
                                                                                        HorizontalAlignment:=horizontal_alignment.Center
                                                                                        VerticalAlignment:=vertical_alignment.Center
                                                                                        Widget:=TextBlock_Tier_Box3
                                                            stack_box_slot: 
                                                                Widget:=overlay:
                                                                    Slots:=array:
                                                                        overlay_slot:
                                                                            HorizontalAlignment:=horizontal_alignment.Fill
                                                                            VerticalAlignment:=vertical_alignment.Fill
                                                                            Widget:=texture_block{DefaultImage:=Images.EmptyTexture,DefaultDesiredSize:=QuietButton_PetSelectorDesiredSize}

                canvas_slot:
                    Anchors:=anchors{Minimum:=vector2{X:=0.5,Y:=0.5},Maximum:=vector2{X:=0.5,Y:=0.5}}
                    ZOrder:=0
                    SizeToContent:=true
                    Offsets:=margin{Left:=0.0,Top:=420.0,Right:=0.0,Bottom:=0.0} #PositionX, PositionY, SizeX, SizeY
                    Alignment:=vector2{X:=0.5,Y:=0.5}
                    Widget:=overlay:
                        Slots:=array:
                            overlay_slot:
                                Padding:=margin{Top:=RowTopHugePadding,Bottom:=RowTopHugePadding}
                                Widget:=stack_box:  
                                    Orientation:=orientation.Vertical
                                        Slots:=array:
                                            stack_box_slot: 
                                                Widget:=stack_box:  
                                                    Orientation:=orientation.Horizontal
                                                    Slots:=array:
                                                        stack_box_slot: 
                                                            Padding:=margin{Right:=SelectorBoxLinePadding,Left:=SelectorBoxLinePadding}
                                                            Widget:=overlay:
                                                                Slots:=array:
                                                                    overlay_slot:
                                                                        Widget:=stack_box:
                                                                            Orientation:=orientation.Horizontal
                                                                            Slots:=array:
                                                                                stack_box_slot: 
                                                                                    Padding:=margin{Left:=SelectorBoxLinePadding,Right:=SelectorBoxLinePadding}
                                                                                    #Padding:=margin{Left:=Icon_LeftMargin}
                                                                                    Widget:=TextureBlock_CostStatsA
                                                                                stack_box_slot: 
                                                                                    Padding:=margin{Left:=SelectorBoxLinePadding,Right:=SelectorBoxLinePadding}
                                                                                    #Padding:=margin{Left:=TextIcon_LeftMargin}
                                                                                    Widget:=TextBlock_CostStatsA
                                                                                stack_box_slot: 
                                                                                    Padding:=margin{Left:=SelectorBoxLinePadding,Right:=SelectorBoxLinePadding}
                                                                                    Widget:=text_block{DefaultText:=BasicMessage(" | Improve stats A"), DefaultTextColor:=CoalTextColor, DefaultShadowOffset:=ShadowOffset, DefaultShadowOpacity := DefaultTextShadowOpacity}
                                                                    overlay_slot:
                                                                        Widget:=texture_block{DefaultImage:=Images.EmptyTexture,DefaultDesiredSize:=vector2{X:=530.0,Y:=80.0}}   
                                                                    overlay_slot:
                                                                        HorizontalAlignment:=horizontal_alignment.Fill
                                                                        VerticalAlignment:=vertical_alignment.Fill
                                                                        Widget:=Upgrade_AStatsButton
                                                                                 
                                                        stack_box_slot: 
                                                            Padding:=margin{Right:=SelectorBoxLinePadding,Left:=SelectorBoxLinePadding}
                                                            Widget:=overlay:
                                                                Slots:=array:
                                                                    overlay_slot:
                                                                        Widget:=stack_box:
                                                                            Orientation:=orientation.Horizontal
                                                                            Slots:=array:
                                                                                stack_box_slot: 
                                                                                    Padding:=margin{Left:=SelectorBoxLinePadding,Right:=SelectorBoxLinePadding}
                                                                                    #Padding:=margin{Left:=Icon_LeftMargin}
                                                                                    Widget:=TextureBlock_CostStatsB
                                                                                stack_box_slot: 
                                                                                    Padding:=margin{Left:=SelectorBoxLinePadding,Right:=SelectorBoxLinePadding}
                                                                                    #Padding:=margin{Left:=TextIcon_LeftMargin}
                                                                                    Widget:=TextBlock_CostStatsB
                                                                                stack_box_slot: 
                                                                                    Padding:=margin{Left:=SelectorBoxLinePadding,Right:=SelectorBoxLinePadding}
                                                                                    Widget:=text_block{DefaultText:=BasicMessage(" | Improve stats B"), DefaultTextColor:=CoalTextColor, DefaultShadowOffset:=ShadowOffset, DefaultShadowOpacity := DefaultTextShadowOpacity}
                                                                    overlay_slot:
                                                                        Widget:=texture_block{DefaultImage:=Images.EmptyTexture,DefaultDesiredSize:=vector2{X:=530.0,Y:=80.0}}  
                                                                    overlay_slot:
                                                                        HorizontalAlignment:=horizontal_alignment.Fill
                                                                        VerticalAlignment:=vertical_alignment.Fill
                                                                        Widget:=Upgrade_BStatsButton
                                                                                  
                                            stack_box_slot: 
                                                    Padding:=margin{Top:=RowTopHugePadding,Bottom:=RowTopHugePadding}
                                                    Widget:=overlay:
                                                        Slots:=array:
                                                            overlay_slot:
                                                                Widget:=texture_block{DefaultImage:=Images.EmptyTexture,DefaultDesiredSize:=vector2{X:=BigCloseButtonSizeX,Y:=80.0}}
                                                            overlay_slot: 
                                                                HorizontalAlignment:=horizontal_alignment.Fill
                                                                VerticalAlignment:=vertical_alignment.Fill
                                                                Widget:=CancelButton
                                    
                                    
                            
                                                           
MakePetUpgradesUI<constructor>(InPetsManager : pets_manager)<transacts> := pets_upgrades_ui :
    PetsManager:=InPetsManager

                                                                                                                