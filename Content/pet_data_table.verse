pet_data_table := class<final><persistable>:
    #Version
    Version:int=1
    IsEquipped:logic=false
    Type:int=0

    Purity:int=1

    LVPet_PackageA:int=1
    LVPet_PackageB:int=1

    #HexaGame
    HexaGamePositionX:int=-1
    HexaGamePositionY:int=-1

MakePetDataTable<constructor>(OldTable:pet_data_table)<transacts> := pet_data_table:
    Version := OldTable.Version
    IsEquipped:=OldTable.IsEquipped
    Type := OldTable.Type

    Purity := OldTable.Purity

    LVPet_PackageA := OldTable.LVPet_PackageA
    LVPet_PackageB := OldTable.LVPet_PackageB

    HexaGamePositionX := OldTable.HexaGamePositionX
    HexaGamePositionY := OldTable.HexaGamePositionY
    #Hexagame
    


