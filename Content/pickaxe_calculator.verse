using { /Verse.org/Simulation }
using { /Verse.org/Random }


    #Critical Prob LV Flag
    AreaLV_Multipliers : []int = array{
        10 #12
        11 #12 
        12 #15
        13 #15
        14 #15
        15 #15
        15 #15
        20 #20 #Desert Entrance
        20 #20
        22 #22
        22 #22
        23 #23
        23 #23
        24 #24
        25 #25 Secret Laboratory
        25 #25 
        25 #25
        25 #25
    }

    LVZones_ManualModifiersCosts : []float = array{
        1.0 #Zone 0
        1.0 #Zone 1
        1.0 #Zone 2
        1.0 #Zone 3
        1.0 #Zone 4
        1.0 #Zone 5
        1.0 #Zone 6
        1.0 #Zone 7
        1.0 #Zone 8
        0.75 #Zone 9
        0.75 #Zone 10
        0.5 #Zone 11
        0.35 #Zone 12
        0.35 #Zone 13
        0.2 #Zone 14
        0.1 #Zone 15
        0.05 #Zone 16
        0.025 #Zone 17
        0.025 #Zone 18
        1.0 #Zone 19
        1.0 #Zone 20
    }


    #HEALTH CALCULATOR
    #Flags https://docs.google.com/spreadsheets/d/1G1X-YGXF_2jjdY69GHdhqIrsZMOHPdfXblED1hLdOEU/edit?gid=0#gid=0

    #AproxMutantMaterialPerHB-Level


    RequiredHBBroken_PerLevel_ForPet : float = 10.0
    
    MinNum_AVGHit : float = 15.0
    MaxNum_AVGHit : float = 225.0
    

    ExtraHitsPerLevel : float = 2.0

    IncrementBetweenLVs : float = 15.0

    CoalProb : float = 0.025 #Every 30 seconds (Aprox 40 hits, 0.025)

pickaxe_calculator := class():

    HardBreadBaseHealth : float = 250.0
    ZoneBaseCost : float = 15.0
    LastZoneAgedSapphiresCost : float = 70.0

    var AllLVs_HardBreadMaxHealth : []float = array{}
    var AllLVs_ZoneCosts : []float = array{}
    var PetPurityExpectedPerHBLevel : []int = array{}
    var PetPackExpectedPerHBLevel : []int = array{}
    var EfficiencyExpectedArray : []float = array{

    }

    <#
    #Critical Prob LV Flag
    LVCriticalProb_ExpectedDamage : []float = array{
        0.67,
        16.59,
        490.93,
        14552.32,
        400449.12,
        4430623.76,
        43479438.75,
        466915589.00,
        5233036890.00,
        58910591620.00,
        501537748936.00,
        4663908769293.00,
        44474679870085.00,
        446642782327337.00,
        4813260000000000.00,
        56539600000000000.00,
        549779000000000000.00,
        4906430000000000000.00,
        44584800000000000000.00,
        427255000000000032768.00,
        4479430000000000065536.00
    }

    #Critical Dam LV
    LVCriticalDam_ExpectedDamage : []float = array{
        0.67,
        68.07,
        7280.36,
        740567.65,
        21594320.13,
        448087604.10,
        10211686302.00,
        305367947856.00,
        8038107726542.00,
        227396609375789.00,
        4627630000000000.00,
        102573000000000000.00,
        2592330000000000000.00,
        72377200000000000000.00,
        2177190000000000065536.00,
        46110400000000002097152.00,
        1028839999999999821742080.00,
        25646300000000001443889152.00,
        663963999999999972997070848.00,
        18072000000000000972407439360.00,
        323538000000000008533798027264.00
    }


    #Weak points
    LVWeakPoint_ExpectedDamage : []float = array{
        0.67,
        26.59,
        751.17,
        14552.32,
        4430623.76,
        153801773.50,
        3665066789.00,
        71867709999.00,
        1296732072968.00,
        23640885411993.00,
        465387917264876.00,
        8616630000000000.00,
        183140000000000000.00
    }

    #>

    GenerateHBHealthControllers(GameManager : game_manager) : void =
        var LastLVHealth : float = 0.0
        var LastZoneCost : float = 0.0
        for(Level:=0..30):
            if(Level=0):
                if(LastLVHealth=0.0):
                    set LastLVHealth = HardBreadBaseHealth
            else:
                if(Multiplier := AreaLV_Multipliers[Level]):
                    set LastLVHealth = LastLVHealth * Multiplier
                    set LastZoneCost = LastZoneCost * Multiplier
                else:
                    if(MaxMultiplier := AreaLV_Multipliers[AreaLV_Multipliers.Length-1]):
                        set LastLVHealth = LastLVHealth * MaxMultiplier
                        set LastZoneCost = LastZoneCost * MaxMultiplier
                    else:
                        PrintFatalError("On GenerateHBHealthControllers")
            if(Level=1):
                set LastZoneCost = ZoneBaseCost
            set AllLVs_HardBreadMaxHealth = AllLVs_HardBreadMaxHealth + array{LastLVHealth}
            #Zone cost set
            ZoneMultiplierCost := 1.0/(Pow(PickaxeTierDegradation,AllLVs_ZoneCosts.Length*1.0)*1.75) #Based on pickaxeefficiency, multiplier per 2, so it is not soo excesive difference
            if(AllLVs_ZoneCosts.Length=0):
                set AllLVs_ZoneCosts = AllLVs_ZoneCosts + array{1.0}
            else if(ManualModifier := LVZones_ManualModifiersCosts[AllLVs_ZoneCosts.Length+1]):
                if(AllLVs_ZoneCosts.Length=HB_MAXLEVEL-1):
                    set AllLVs_ZoneCosts = AllLVs_ZoneCosts + array{LastZoneAgedSapphiresCost}
                else:
                    var ModMultiplier : float = 1.0
                    if(NewMultiplier := GameManager.AllManualAreaCostMultipliers[AllLVs_ZoneCosts.Length]):
                        set ModMultiplier = NewMultiplier
                        Log("Zone {AllLVs_ZoneCosts.Length} extra mutiplier: {ZoneMultiplierCost}, Manual Modifier: {ManualModifier}",LOG_LEVEL_DEBUG8)
                    set AllLVs_ZoneCosts = AllLVs_ZoneCosts + array{LastZoneCost*ModMultiplier*ZoneMultiplierCost*ManualModifier}


            <#
        for(AreaLevel:=0..20):
            
            PickaxeEfficiency_Multiplier := MainFormula_PickaxeEfficiency(AreaLevel*1.0)
            HBTierEfficiency_Multiplier := HardBread_Tier_Multiplier*AreaLevel
            
            PetsMultiplier_Multiplier:=ALL_PLAYER_PICKEFFICIENCY_MULTIPLIER*300.0*3.0

            #CoalEffect_Multiplier := NOT USED

            BakersEfficiency := MainFormulaBaker_Efficiency(AreaLevel*1.0)

           
            
        
       
            LevelEfficiencyExpected := MaxEfficiencyMultiplier/(20.0-(Index*1.0))
            Log("LevelEfficiencyExpected {GetOptimizedMagnitudeLetter(LevelEfficiencyExpected)}",LOG_LEVEL_DEBUG4)
            set EfficiencyExpectedArray = EfficiencyExpectedArray + array{LevelEfficiencyExpected} #>

        <#
        AvgNum_AVGHit:=(MinNum_AVGHit+MaxNum_AVGHit)/2.0
        if(IntAvgNum_AVGHit:=Floor[AvgNum_AVGHit]):
            for(Level:=0..30):
                #Purity
                MutantMaterialAward:=CheckFinalMutantMaterialAward(Level,AvgNum_AVGHit)
                TotalMutantMaterialperLevel := MutantMaterialAward * RequiredHBBroken_PerLevel_ForPet
                FloatPurityLevel:=MainFormula_PetPurityCost(0,?AmountMutantMaterial:=TotalMutantMaterialperLevel,?Inverted:=true)
                if(PurityLevel := Floor[FloatPurityLevel]):
                    set PetPurityExpectedPerHBLevel = PetPurityExpectedPerHBLevel + array{PurityLevel}
                #LevelPack
                CoalAward:=CheckFinalCoalAward(Level,IntAvgNum_AVGHit)+CoalProb #CoalProb is the always amount that is disposed
                TotalCoalperLevel := MutantMaterialAward * RequiredHBBroken_PerLevel_ForPet
                FloatPackLevel:=MainFormula_PackageACost(0,?AmountCoal:=TotalMutantMaterialperLevel,?Inverted:=true)
                if(PackLevel := Floor[FloatPackLevel]):
                    set PetPackExpectedPerHBLevel = PetPackExpectedPerHBLevel + array{PackLevel}#>

    MainFormula_GenerateHardBreadHealth(Level : int, Type : int) : float =
       var HardBreadHealth : float = 999999999999999999999999999999999999999999999999999999999.9
        if(AssignedHealth:=AllLVs_HardBreadMaxHealth[Level]):
            return AssignedHealth
        return HardBreadHealth

    ConvertDamageToSapphires(Damage : float, HBLevel : int) : float =
        var TotalResourcers : float = Damage
        if(EfficiencyMultiplier:=EfficiencyExpectedArray[HBLevel]):
            Log("EfficiencyMultiplier {GetOptimizedMagnitudeLetter(EfficiencyMultiplier)} single Hit Sapphires : ",LOG_LEVEL_DEBUG4)
            set TotalResourcers = TotalResourcers * EfficiencyMultiplier
        TotalSaphires := TotalResourcers / Req_Resourcers_PerSapphire
        ##Print("MaxEfficiencyMultiplier: {MaxEfficiencyMultiplier}")
        return TotalSaphires


        <#
    MainFormula_GenerateHardBreadHealthOld(Level: int, Type : int<#, ?GetCost : logic = false#>) : float =
        <#Health explanation
        -On base, each level has a difference of x15. This is to get over 15-225 AVGHit per level
        -One AVGHit has the average critical, critical prob, weak damage and pets enhacements calculated
        -Full calculation: https://docs.google.com/spreadsheets/d/1G1X-YGXF_2jjdY69GHdhqIrsZMOHPdfXblED1hLdOEU/edit?gid=0#gid=0
        -For each level, +3 AVG_Hit are added to the health
        #>
    
        AvgNum_AVGHit:=(MinNum_AVGHit+MaxNum_AVGHit)/2.0
        BaseHealth:=15.0*Pow(IncrementBetweenLVs,Level*1.0) ##MAIN FORMULA BASE HEALTH
        BaseDamagePerHit_Expected := BaseHealth/AvgNum_AVGHit

        #if(GetCost?):
        #    BaseDamagePerHit_Expected
        

        #Cantidad de daño base que aproximadamente estaría haciendo con el pico por nivel de perk
        #CALCULE EXTRAS
        #LV_CriticalProb
        var Index_CriticalProb : int = 0
        loop:
            if(LevelDamage:=LVCriticalProb_ExpectedDamage[Index_CriticalProb], LevelDamage >= BaseDamagePerHit_Expected):
                break
            else if(Index_CriticalProb=LVCriticalProb_ExpectedDamage.Length-1):
                break
            if(Index_CriticalProb>LVCriticalProb_ExpectedDamage.Length-1):
                set Index_CriticalProb = LVCriticalProb_ExpectedDamage.Length-1
                break
            set Index_CriticalProb = Index_CriticalProb + 1
        #LV_CriticalDamage
        var Index_CriticalDamage : int = 0
        loop:
            if(LevelDamage:=LVCriticalDam_ExpectedDamage[Index_CriticalDamage], LevelDamage >= BaseDamagePerHit_Expected):
                break
            else if(Index_CriticalDamage=LVCriticalDam_ExpectedDamage.Length-1):
                break
            if(Index_CriticalDamage>LVCriticalDam_ExpectedDamage.Length-1):
                set Index_CriticalDamage = LVCriticalDam_ExpectedDamage.Length-1
                break
            set Index_CriticalDamage = Index_CriticalDamage + 1
        #LV_WeakPoint
        var Index_WeakPoint : int = 0
        loop:
            if(LevelDamage:=LVWeakPoint_ExpectedDamage[Index_WeakPoint], LevelDamage >= BaseDamagePerHit_Expected):
                break
            else if(Index_WeakPoint=LVWeakPoint_ExpectedDamage.Length-1):
                break
            if(Index_WeakPoint>LVWeakPoint_ExpectedDamage.Length-1):
                set Index_WeakPoint = LVWeakPoint_ExpectedDamage.Length-1
                break
            set Index_WeakPoint = Index_WeakPoint + 1

        #PETS
        var Pets_TotalPickDamageMultiplier : float = 1.0
        var Pets_TotalWeakPointMultiplier : float = 1.0
        var Pets_TotalCriticalMultiplier : float = 1.0
        var NumExpectedPets : int = 0
        #Numpets
        if(NumPets := Floor(Level/5)):
            set NumExpectedPets = NumPets
            if(NumExpectedPets>3):
                set NumExpectedPets = 3
        #Pets expected stats
        if:
            NumExpectedPets>0
            PetPurityExpected:=PetPurityExpectedPerHBLevel[Level]
            PetPackExpected:=PetPackExpectedPerHBLevel[Level]
        then:
            set Pets_TotalPickDamageMultiplier = Pets_TotalPickDamageMultiplier + ((MainFormula_EnhancementsPetStats(STAT_TYPE_PLAYER_PICKDAMAGE,-1,PetPackExpected,PetPurityExpected)*NumExpectedPets)-1.0)
            set Pets_TotalWeakPointMultiplier = Pets_TotalWeakPointMultiplier + ((MainFormula_EnhancementsPetStats(STAT_TYPE_PLAYER_WEAKPOINTMULT,-1,PetPackExpected,PetPurityExpected)*NumExpectedPets)-1.0)
            set Pets_TotalCriticalMultiplier = Pets_TotalCriticalMultiplier + ((MainFormula_EnhancementsPetStats(STAT_TYPE_PLAYER_PICKCRITICALMULT,-1,PetPackExpected,PetPurityExpected)*NumExpectedPets)-1.0)

        #PrintWarning("TO DO Calculate Pets improvements")
        PickaxeCriticalProb:=MainFormula_PickaxeCriticalProb(Index_CriticalProb)
        PickaxeCriticalDamage:=MainFormula_PickaxeCriticalDamage(Index_CriticalDamage)
        WeakPointMult:=MainFormula_WeakPointMult(Index_WeakPoint)+1.0
        
        TotalPickaxeTotalStrength:=BaseDamagePerHit_Expected*Pets_TotalPickDamageMultiplier
        TotalPickaxeWeakPointMult:=WeakPointMult+(Pets_TotalWeakPointMultiplier-1.0)
        TotalPickaxeCriticalDamage:=PickaxeCriticalDamage+(Pets_TotalCriticalMultiplier-1.0)

        ##Print("BaseDamagePerHit_Expected {BaseDamagePerHit_Expected} WeakPointMult: {WeakPointMult}, PickaxeCriticalProb {PickaxeCriticalProb}, PickaxeCriticalDamage {PickaxeCriticalDamage}")
        TupleFinalDamagePerHit_Expected := GetPickaxeDamage(true,TotalPickaxeTotalStrength,false,PickaxeCriticalProb,TotalPickaxeCriticalDamage,TotalPickaxeWeakPointMult)
        ExtraHits:=ExtraHitsPerLevel*Level
        FinalTotalHealth:=TupleFinalDamagePerHit_Expected(0)*(AvgNum_AVGHit+ExtraHits)
        
        return FinalTotalHealth
        #>

MainFormula_PickaxeEfficiency(Level : float) : float =
    if(Level <= 0.0):
        return PickaxeBaseEfficiency / 2.0
    else:
        PickaxeBaseEfficiency * Level


GetPickaxeDamage(MustGetAVG : logic, PickaxeTotalStrength : float, IsWeakPoint : logic, CriticalProb : float, CriticalDamage : float, WeakPointMult : float, PetsPickaxeMultiplier : float) : tuple(float,logic) =
    var TotalDamage : float = PickaxeTotalStrength
    var IsCritical : logic = false
    if(not MustGetAVG?):
        CriticalRaffle := GetRandomFloat(0.0,1.0)
        set TotalDamage = TotalDamage +  PickaxeTotalStrength * (PetsPickaxeMultiplier-1.0)
        if(CriticalRaffle<=CriticalProb):
            set IsCritical = true
            set TotalDamage = TotalDamage + PickaxeTotalStrength * (CriticalDamage-1.0)
        if(IsWeakPoint?):
            set TotalDamage = TotalDamage + PickaxeTotalStrength * (WeakPointMult-1.0)
        #Todo pet modifiers
    else: 
        AvgNumHitsExpected_ToWeakPoint := ((MinNum_AVGHit+MaxNum_AVGHit)/2.0)*PercentageToWeakPoint#15-225 average: 120 hits -> 100/ 24 hits
        <#
        if(94900.0<PickaxeTotalStrength<95900.0):
            
            ##Print("################")
            #Print("PickaxeTotalStrength: {PickaxeTotalStrength}")
            #Print("AvgNumHitsExpected_ToWeakPoint: {AvgNumHitsExpected_ToWeakPoint}")
            #Print("CriticalProb: {CriticalProb}, CriticalDamage: {CriticalDamage}, AvgCriticalDamage added: {(TotalDamage*CriticalDamage*CriticalProb)}")
            set TotalDamage = TotalDamage + (TotalDamage*CriticalDamage*CriticalProb)
            #Print("TotalDamage: {TotalDamage}, WeakPointMult: {WeakPointMult}")
            #Print("AvgWeakPoint added: {((TotalDamage*WeakPointMult)/AvgNumHitsExpected_ToWeakPoint)}")
            set TotalDamage = TotalDamage + ((TotalDamage*WeakPointMult)/AvgNumHitsExpected_ToWeakPoint)
        else:#>
        set TotalDamage = TotalDamage +  PickaxeTotalStrength * (PetsPickaxeMultiplier-1.0)
        set TotalDamage = TotalDamage + (PickaxeTotalStrength * CriticalDamage*CriticalProb)
        set TotalDamage = TotalDamage + ((PickaxeTotalStrength * WeakPointMult)/AvgNumHitsExpected_ToWeakPoint)
        #Todo pet modifiers
    return (TotalDamage,IsCritical)

IsThereCoalAward() : logic =
    RandomResult:=GetRandomFloat(0.0,1.0)
    if(CoalProb>RandomResult):
        return true
    else:
        return false

CheckFinalCoalAward(HardBreadLV : int, NumHitsAccumulated : int, DamagePercentage : float) : float =
    var FinalCoalAmount : float = 0.0
    if(HardBreadLV>1):
        if(Amount := Floor[((NumHitsAccumulated*1.0/MinNum_AVGHit)*(HBLevelPer_Coal*0.1))*DamagePercentage]):
            set FinalCoalAmount = Amount*1.0
    return FinalCoalAmount

CheckFinalMutantMaterialAward(HardBreadLV : int, NumHitsAccumulated : float, DamagePercentage : float, ?MobType : int = -1, ?NumDungeon : int = -1) : float =
    var FinalMutantMaterialAmount : float = 0.0
    var MutantAmount : float = Pow(2.0,HardBreadLV*1.5)
    if(MobType=-1):
        if(NumHitsAccumulated>MinNum_AVGHit):
            set FinalMutantMaterialAmount = MutantAmount + ((MutantAmount/100.0)*NumHitsAccumulated)
        else:
            if(NumHitsAccumulated>4.0):
                Multiplier := (NumHitsAccumulated/MinNum_AVGHit) * 0.5
                set MutantAmount = MutantAmount * Multiplier
                set FinalMutantMaterialAmount = MutantAmount + ((MutantAmount/100.0)*NumHitsAccumulated) * Multiplier
        set FinalMutantMaterialAmount = FinalMutantMaterialAmount * DamagePercentage
    else:
        set FinalMutantMaterialAmount = MutantAmount + ((MutantAmount/100.0)*(MinNum_AVGHit+1.0))
        if(NumDungeon<>-1):
            if(MobType=MINIBOSS_MOB):
                set FinalMutantMaterialAmount = FinalMutantMaterialAmount*Reward_NormalMobMultiplier*Reward_MiniBossMultiplier
            else if(MobType=GROUND_MOB):
                set FinalMutantMaterialAmount = FinalMutantMaterialAmount*Reward_DungeonMobMultiplier*Reward_GroundMobMultiplier
            else if(MobType=DRUID_MOB):
                set FinalMutantMaterialAmount = FinalMutantMaterialAmount*Reward_DungeonMobMultiplier*Reward_GroundMobMultiplier
        else:
            if(MobType=MINIBOSS_MOB):
                set FinalMutantMaterialAmount = FinalMutantMaterialAmount*Reward_NormalMobMultiplier*Reward_MiniBossMultiplier
            else if(MobType=GROUND_MOB):
                set FinalMutantMaterialAmount = FinalMutantMaterialAmount*Reward_NormalMobMultiplier*Reward_GroundMobMultiplier
            else if(MobType=DRUID_MOB):
                set FinalMutantMaterialAmount = FinalMutantMaterialAmount*Reward_NormalMobMultiplier*Reward_GroundMobMultiplier


    return FinalMutantMaterialAmount

CalculeAFKReward(Bakery : bakery_machines_manager) : float =
    return (CheckFinalMutantMaterialAward(Bakery.PlayerDataCache.LV_Area,MinNum_AVGHit+1.0,1.0)/MinNum_AVGHit)*0.1


#Auto Generation
MainFormula_AutoResource(Level : int) : float =
    if(Level=0):
        return 0.05
    return Level * 0.15

#Pickaxe strength
MainFormula_TotalPickaxeStrength(TLV : []int) : float =
    var Total : float = 0.0
    for(Tier := 0..TLV.Length-1, Level := TLV[Tier], Level <> -1):
        Cost:=MainFormula_PickaxeCostStrength(PickaxeBaseStrenght, Tier*1.0, Level*1.0)
        TierProduction := MainFormula_PickaxeStrength(Cost, Level*1.0, Tier)
        Log("PickaxeStrength / Cost: {Cost} TierProduction: {TierProduction}", LOG_LEVEL_UIs)
        set Total = Total + TierProduction
    Log("PickaxeStrength / Total: {Total}", LOG_LEVEL_UIs)
    return Total

MainFormula_PickaxeStrength(Cost : float, Level : float, Tier : int) : float =
    if(Level=-1.0 or Level=0.0):
        return 0.5
    else:
        var Ammount : float = 0.0
        if(Cost<1000.0):
            set Ammount = Cost
        else:
            set Ammount = Cost * (Pow(PickaxeLevelDegradation,Level)) * (Pow(PickaxeTierDegradation,Tier*1.0))
        return Ammount

MainFormula_PickaxeCostStrength(CostBase : float, Tier : float, Level : float) : float = 
    var Ammount : float = Pow(Level,2.0)*Pow(CostBase,Tier)
    if(Tier >= 0.0):
        set Ammount = Ammount + 10.0*Pow(CostBase,Tier-1.0)
    if(Ammount<1000.0):
        ##Print("pre Amount: {Ammount}")
        if(IntAmmount := Ceil[Ammount], set Ammount = IntAmmount * 1.0):
    ##Print("Tier {Tier}")
    ##Print("Level {Level}")

    ##Print("CostBase: {CostBase}")
    ##Print("Amount: {Ammount}")
    return Ammount

MainFormula_PickaxeCostEfficiency(ResourceType : int, CurrentLevel : float) : float =
    Ammount:=Pow(PickaxeBaseEfficiencyCost,CurrentLevel)
    return Ammount



#Critical Damage 
GetMaxPickaxeCriticalDamageLV() : float =
    return MaxPickaxeCriticalDamage/PickaxeCriticalDamageBase

MainFormula_PickaxeCriticalDamage(Level : int) : float =
    if(Level=0):
        return PickaxeCriticalDamageBase/2.0
    else:
        return Level * PickaxeCriticalDamageBase

MainFormula_PickaxeCriticalDamage_Cost(Level : int) : float =
    if(Level = 0):
        return 5.0
    else:
        return Pow(PickaxeCriticalDamageCostBase,Level*1.0)

#Critical probability
GetMaxPickaxeCriticalProbLV() : float =
    return MaxPickaxeCriticalProb/PickaxeCriticalProbBase

MainFormula_PickaxeCriticalProb(Level : int) : float =
    return Level * PickaxeCriticalProbBase

MainFormula_PickaxeCriticalProb_Cost(Level : int) : float =
    if(Level = 0):
        return 5.0
    else:
        ##Print("Level: {Level}, amount: {Pow(PickaxeCriticalProbCostBase,Level*1.0)}")
        return Pow(PickaxeCriticalProbCostBase,Level*1.0)

#Luck - Coal
GetMaxLuckLV() : float =
    return MaxCoalProbability/CoalProbabilityBase

MainFormula_CoalProb(Level : int) : float =
    return Level * CoalProbabilityBase

MainFormula_CoalProbCost(Level : int) : float =
    return Pow(15.0,Level*1.0)

#Weak point multiplier
GetMaxWeakPointLV() : float =
    return MaxWeakPointMult/WeakPointMultBase

MainFormula_WeakPointMult(Level : int) : float =
    if(Level = 0):
        return 0.1
    else:
        return Level * WeakPointMultBase

MainFormula_WeakPointMultCost(Level : int) : float =
    return Pow(WeakPointCostBase,Level*1.0)

#MultiResource 
MainFormula_MultiResource (Level : int) : float =
    return Level * 1.0

MainFormula_MultiResourceCost(Level : int) : float =
    return Pow(MultipleResourceGatherCostBase,Level*1.0)
    
#Backpack limit
CalculeBackpackUpgrade<public>(CurrentLevel : int, Resource : int) : tuple(float,float,float) =
    var Base : float = Base_BackpackCostLimit
    #### Next stats
    var ObjetiveLevel : int = CurrentLevel+1
    CurrentCost :=MainFormulaBackpackCost_Limit(Base, CurrentLevel*1.0)
    ObjetiveCost :=MainFormulaBackpackCost_Limit(Base, ObjetiveLevel*1.0)

    CurrentLimit := MainFormulaBackpack_Limit(CurrentCost)
    NextLimit  := MainFormulaBackpack_Limit(ObjetiveCost)

    #### Difference
    TotalDifference := NextLimit-CurrentLimit
    return(CurrentLimit,ObjetiveCost,TotalDifference)
    

CalculeStrengthLimitLevel(BaseSpeed : float) : int =
    var MaxLevel : int = 0
    if(Value := Floor[BaseSpeed / MinBakeTime]):
        set MaxLevel = Value - 1
    return MaxLevel

CalculePickaxeUpgradeDamage<public>(Tier : int, CurrentLevel : int, BuyModeMultiplier : int) : tuple(float,float,float) =
    #### Next stats
    var ObjetiveLevel : int = CurrentLevel+BuyModeMultiplier
    if(ObjetiveLevel > MaxTierLevel):
        set ObjetiveLevel = MaxTierLevel
    var TotalNextCost : float = 0.0
    for(Level:=CurrentLevel..ObjetiveLevel-1):
        set TotalNextCost = TotalNextCost + MainFormula_PickaxeCostStrength(Base_PickaxeStrength, Tier*1.0, Level*1.0)

    ObjetiveCost := MainFormula_PickaxeCostStrength(Base_PickaxeStrength, Tier*1.0, ObjetiveLevel*1.0)
    NextProduction := MainFormula_PickaxeStrength(ObjetiveCost, ObjetiveLevel*1.0,Tier)
    #### Current stats
    CurrentCost:=MainFormula_PickaxeCostStrength(Base_PickaxeStrength, Tier*1.0, CurrentLevel*1.0)
    CurrentProduction := MainFormula_PickaxeStrength(CurrentCost,CurrentLevel*1.0,Tier)

    #### Difference
    TotalDifference := NextProduction-CurrentProduction
    return(CurrentProduction,TotalNextCost,TotalDifference)

CalculePickaxeResourceEfficiency<public>(CurrentLevel : int, ResourceType : int, BuyModeMultiplier : int) : tuple(float,float,float) =
    #### Next stats
    var ObjetiveLevel : int = CurrentLevel+BuyModeMultiplier
    if(ObjetiveLevel > PickaxeMaxEfficiencyLV):
        set ObjetiveLevel = PickaxeMaxEfficiencyLV
    var TotalNextCost : float = 0.0
    for(Level:=CurrentLevel..ObjetiveLevel-1):
        set TotalNextCost = TotalNextCost + MainFormula_PickaxeCostEfficiency(ResourceType, Level*1.0)

    ObjetiveCost := MainFormula_PickaxeCostEfficiency(ResourceType, ObjetiveLevel*1.0)
    NextProduction := MainFormula_PickaxeEfficiency(ObjetiveLevel*1.0)
    #### Current stats
    CurrentCost:=MainFormula_PickaxeCostEfficiency(ResourceType, CurrentLevel*1.0)
    CurrentProduction := MainFormula_PickaxeEfficiency(CurrentLevel*1.0)

    #### Difference
    TotalDifference := NextProduction-CurrentProduction
    return(CurrentProduction,TotalNextCost,TotalDifference)
    

CalculeTotalStrength<public>(TLV : []int) : float = 
    var Total : float = 0.0 
    for(Tier := 0..TLV.Length-1, Level := TLV[Tier], Level <> -1):

        var LevelToCalculate : int = Level
        if(Level>MaxTierLevel):
            set LevelToCalculate = MaxTierLevel

        Cost:=MainFormula_PickaxeCostStrength(Base_PickaxeStrength, Tier*1.0, LevelToCalculate*1.0)
        TierProduction := MainFormula_PickaxeStrength(Cost, LevelToCalculate*1.0, Tier)

        set Total = Total + TierProduction
    return Total

#CriticalProbUpgrade
CalculePickaxeCriticalProbUpgrade<public>(CurrentLevel : int) : tuple(float,float,float) =
    #### Next stats
    var ObjetiveLevel : int = CurrentLevel+1

    ObjetiveCost := MainFormula_PickaxeCriticalProb_Cost(ObjetiveLevel)
    NextProduction := MainFormula_PickaxeCriticalProb(ObjetiveLevel)
    #### Current stats
    CurrentCost:=MainFormula_PickaxeCriticalProb_Cost(CurrentLevel)
    CurrentProduction := MainFormula_PickaxeCriticalProb(CurrentLevel)

    #### Difference
    TotalDifference := NextProduction-CurrentProduction
    return(CurrentProduction,ObjetiveCost,TotalDifference)

#CriticalDamageUpgrade
CalculePickaxeCriticalDamageUpgrade<public>(CurrentLevel : int) : tuple(float,float,float) =
    #### Next stats
    var ObjetiveLevel : int = CurrentLevel+1

    ObjetiveCost := MainFormula_PickaxeCriticalDamage_Cost(ObjetiveLevel)
    NextProduction := MainFormula_PickaxeCriticalDamage(ObjetiveLevel)+1.0
    #### Current stats
    CurrentCost:=MainFormula_PickaxeCriticalDamage_Cost(CurrentLevel)
    CurrentProduction := MainFormula_PickaxeCriticalDamage(CurrentLevel)+1.0

    #### Difference
    TotalDifference := NextProduction-CurrentProduction
    return(CurrentProduction,ObjetiveCost,TotalDifference)

#Luck - Coal
CalculePickaxeLuckUpgrade<public>(CurrentLevel : int) : tuple(float,float,float) =
    #### Next stats
    var ObjetiveLevel : int = CurrentLevel+1

    ObjetiveCost := MainFormula_CoalProbCost(ObjetiveLevel)
    NextProduction := MainFormula_CoalProb(ObjetiveLevel)
    #### Current stats
    CurrentCost:=MainFormula_CoalProbCost(CurrentLevel)
    CurrentProduction := MainFormula_CoalProb(CurrentLevel)

    #### Difference
    TotalDifference := NextProduction-CurrentProduction
    return(CurrentProduction,ObjetiveCost,TotalDifference)

#Weak Point
CalculePickaxeWeakUpgrade<public>(CurrentLevel : int) : tuple(float,float,float) =
    #### Next stats
    var ObjetiveLevel : int = CurrentLevel+1

    ObjetiveCost := MainFormula_WeakPointMultCost(ObjetiveLevel)
    NextProduction := MainFormula_WeakPointMult(ObjetiveLevel)+1.0
    #### Current stats
    CurrentCost:=MainFormula_WeakPointMultCost(CurrentLevel)
    CurrentProduction := MainFormula_WeakPointMult(CurrentLevel)+1.0

    #### Difference
    TotalDifference := NextProduction-CurrentProduction
    return(CurrentProduction,ObjetiveCost,TotalDifference)


#Multiple resource
CalculeMultipleResUpgrade<public>(CurrentLevel : int) : tuple(float,float,float) =
    #### Next stats
    var ObjetiveLevel : int = CurrentLevel+1

    ObjetiveCost := MainFormula_MultiResourceCost(ObjetiveLevel)
    NextProduction := MainFormula_MultiResource(ObjetiveLevel)
    #### Current stats
    CurrentCost:=MainFormula_MultiResourceCost(CurrentLevel)
    CurrentProduction := MainFormula_MultiResource(CurrentLevel)

    #### Difference
    TotalDifference := NextProduction-CurrentProduction
    return(CurrentProduction,ObjetiveCost,TotalDifference)




    
                            
      

    

    


