using { /Fortnite.com/Devices }
using { /Verse.org/Simulation }
using { /UnrealEngine.com/Temporary/Diagnostics }
# A persistable stat that persists across play sessions.
player_data := class<final><persistable>:
    # The current value of the player's stat.
    Value : int = 0

# Creates a new player_stat class with the given NewValue and a source player_stat.
# The HighestValue is set to either the NewValue or the HighestValue
# from the source player_stat.
MakeTestPlayerData<constructor>(OldStats:player_data, NewValue:int)<transacts> := player_data:
    Value := NewValue