using { /Fortnite.com/Devices }
using { /Verse.org/Simulation }
using { /UnrealEngine.com/Temporary/Diagnostics }

# Represents an abstract stat to update using RecordPlayerStat.
# The class has the unique specifier to make instances of the class comparable.
# The class has the computes specifier to be able to instantiate it at module-scope.
# The class has the abstract specifier so it cannot be instantiated directly, and
# requires subclasses to implement any non-initialized functions, like DebugString().

# Maps players to a table of their player stats.
var AllPlayersDataMap : weak_map(player, player_data_table) = map{}

#EnoughStates
NOT_RES_STATE : int = -1
BACKPACK_RES_STATE : int = 0
MAIN_RES_STATE : int = 1
SECOND_RES_STATE : int = 2

# Manages and updates player_stat_tables for each player.
player_data_manager := class():
    SaveNewData(Player : player, OldData : player_data_table, NewData : player_data_table)<transacts> : void =
        if(Player.IsActive[], FitsInPlayerMap[NewData]):
            if(set AllPlayersDataMap[Player] = NewData):
        else if(Player.IsActive[], set AllPlayersDataMap[Player] = OldData):

    # Return the player_stats_table for the provided Agent.
    LoadPlayerCache(Player:player)<decides><transacts>:player_data_table=
        var PlayerData:player_data_table = player_data_table{}
        if:
            Player.IsActive[]
            PlayerDataTable := AllPlayersDataMap[Player]
            set PlayerData = MakePlayerDataTable(PlayerDataTable)
        return PlayerData

    # Return the player_stats_table for the provided Agent.
    GetPlayerData<private>(Player:player)<decides><transacts>:player_data_table=
        var PlayerData:player_data_table = player_data_table{}
        if:
            Player.IsActive[]
            PlayerDataTable := AllPlayersDataMap[Player]
            set PlayerData = MakePlayerDataTable(PlayerDataTable)
        return PlayerData

    # Initialize stats for all current players.
    InitializeAllPlayers(Players:[]player):void =
        for (Player : Players):
            InitializePlayer(Player)

    # Initialize stats for the given player.
    InitializePlayer(Player:player):void=
        if:
            not AllPlayersDataMap[Player]
            set AllPlayersDataMap[Player] = player_data_table{}

    #SETTERS
    SetVersion(ProfileData : player_data_table, NewVersion:int)<transacts>:player_data_table =
        NewData : player_data_table = player_data_table:
            MakePlayerDataTable<constructor>(ProfileData)
            Version := NewVersion
    SetLanguage(ProfileData : player_data_table, Language:int)<transacts>:player_data_table =
        NewData : player_data_table = player_data_table:
            MakePlayerDataTable<constructor>(ProfileData)
            Language := Language


    SetProductOnDrone(PlayerData : player_data_table, B_On_Drone : float, CB_On_Drone : float, PCS_On_Drone : float) : player_data_table =
        NewData : player_data_table = player_data_table:
            MakePlayerDataTable<constructor>(PlayerData)
            B_On_Drone := B_On_Drone
            CB_On_Drone := CB_On_Drone
            PCS_On_Drone := PCS_On_Drone
        return NewData

    SetMainDeposits(PlayerData : player_data_table, MainWaterDeposit_Value : float, MainFlourDeposit_Value : float, MainCheeseDeposit_Value : float, MainPimentosDeposit_Value : float) : player_data_table =
        NewData : player_data_table = player_data_table:
            MakePlayerDataTable<constructor>(PlayerData)
            MainWaterDeposit_Value := MainWaterDeposit_Value
            MainFlourDeposit_Value := MainFlourDeposit_Value
            MainCheeseDeposit_Value := MainCheeseDeposit_Value
            MainPimentosDeposit_Value := MainPimentosDeposit_Value
        return NewData

    #Main deposits
    SetWaterMainDeposit(PlayerData : player_data_table, MainWaterDeposit_Value : float) : player_data_table =
        NewData : player_data_table = player_data_table:
            MakePlayerDataTable<constructor>(PlayerData)
            MainWaterDeposit_Value := MainWaterDeposit_Value
        return NewData

    SetFlourMainDeposit(PlayerData : player_data_table, MainFlourDeposit_Value : float) : player_data_table =
        NewData : player_data_table = player_data_table:
            MakePlayerDataTable<constructor>(PlayerData)
            MainFlourDeposit_Value := MainFlourDeposit_Value
        return NewData
    
    SetCheeseMainDeposit(PlayerData : player_data_table, MainCheeseDeposit_Value : float) : player_data_table =
            NewData : player_data_table = player_data_table:
                MakePlayerDataTable<constructor>(PlayerData)
                MainCheeseDeposit_Value := MainCheeseDeposit_Value
            return NewData
    
    SetPimentosMainDeposit(PlayerData : player_data_table, MainPimentosDeposit_Value : float) : player_data_table =
            NewData : player_data_table = player_data_table:
                MakePlayerDataTable<constructor>(PlayerData)
                MainPimentosDeposit_Value := MainPimentosDeposit_Value
            return NewData
    
    #Sub deposits
    SetAfterB_SubDeposits(PlayerData : player_data_table, WaterPreparedFor_B : float,FlourPreparedFor_B : float) : player_data_table =
        NewData : player_data_table = player_data_table:
            MakePlayerDataTable<constructor>(PlayerData)
            WaterPreparedFor_B := WaterPreparedFor_B
            FlourPreparedFor_B := FlourPreparedFor_B
        return NewData

    SetAfterCB_SubDeposits(PlayerData : player_data_table, WaterPreparedFor_CB : float, FlourPreparedFor_CB : float,  CheesePreparedFor_CB : float) : player_data_table =
        NewData : player_data_table = player_data_table:
            MakePlayerDataTable<constructor>(PlayerData)
            WaterPreparedFor_CB := WaterPreparedFor_CB
            FlourPreparedFor_CB := FlourPreparedFor_CB
            CheesePreparedFor_CB := CheesePreparedFor_CB
        return NewData

    SetAfterPCS_SubDeposits(PlayerData : player_data_table, WaterPreparedFor_PCS : float, FlourPreparedFor_PCS : float,  CheesePreparedFor_PCS : float, PimentosPreparedFor_PCS : float) : player_data_table =
        NewData : player_data_table = player_data_table:
            MakePlayerDataTable<constructor>(PlayerData)
            WaterPreparedFor_PCS := WaterPreparedFor_PCS
            FlourPreparedFor_PCS := FlourPreparedFor_PCS
            CheesePreparedFor_PCS := CheesePreparedFor_PCS
            PimentosPreparedFor_PCS := PimentosPreparedFor_PCS
        return NewData

    SetWaterSubDeposit(PlayerData : player_data_table, WaterPreparedFor_B : float, WaterPreparedFor_CB : float, WaterPreparedFor_PCS : float) : player_data_table =
        NewData : player_data_table = player_data_table:
            MakePlayerDataTable<constructor>(PlayerData)
            WaterPreparedFor_B := WaterPreparedFor_B
            WaterPreparedFor_CB := WaterPreparedFor_CB
            WaterPreparedFor_PCS := WaterPreparedFor_PCS
        return NewData

    SetFlourSubDeposit(PlayerData : player_data_table, FlourPreparedFor_B : float, FlourPreparedFor_CB : float, FlourPreparedFor_PCS : float) : player_data_table =
        NewData : player_data_table = player_data_table:
            MakePlayerDataTable<constructor>(PlayerData)
            FlourPreparedFor_B := FlourPreparedFor_B
            FlourPreparedFor_CB := FlourPreparedFor_CB
            FlourPreparedFor_PCS := FlourPreparedFor_PCS
        return NewData

    SetCheeseSubDeposit(PlayerData : player_data_table, CheesePreparedFor_CB : float, CheesePreparedFor_PCS : float) : player_data_table =
        NewData : player_data_table = player_data_table:
            MakePlayerDataTable<constructor>(PlayerData)
            CheesePreparedFor_CB := CheesePreparedFor_CB
            CheesePreparedFor_PCS := CheesePreparedFor_PCS
        return NewData

    SetPimentoSubDeposit(PlayerData : player_data_table, PimentosPreparedFor_PCS : float) : player_data_table =
        NewData : player_data_table = player_data_table:
            MakePlayerDataTable<constructor>(PlayerData)
            PimentosPreparedFor_PCS := PimentosPreparedFor_PCS
        return NewData

    SetSuperEnergy(PlayerData : player_data_table, SuperEnergy_Owned : float) : player_data_table =
        NewData : player_data_table = player_data_table:
            MakePlayerDataTable<constructor>(PlayerData)
            SuperEnergy_Owned := SuperEnergy_Owned
        return NewData


        <#
    SetBakeryCacheData(Player : player, OldData : player_data_table, NewData : player_data_table, BakeryMachinesManager : bakery_machines_manager)<transacts> : void =
        NewUpdatedData : player_data_table = player_data_table:
            MakePlayerDataTable<constructor>(NewData)
            #Super Energy
            SuperEnergy_Owned := BakeryMachinesManager.SuperEnergy_Owned
            #Products on drone
            B_On_Drone := BakeryMachinesManager.B_On_Drone
            CB_On_Drone := BakeryMachinesManager.CB_On_Drone
            PCS_On_Drone := BakeryMachinesManager.PCS_On_Drone
            #Deposits
            MainWaterDeposit_Value := BakeryMachinesManager.MainWaterDeposit_Value
            MainFlourDeposit_Value := BakeryMachinesManager.MainFlourDeposit_Value
            MainCheeseDeposit_Value := BakeryMachinesManager.MainCheeseDeposit_Value
            MainPimentosDeposit_Value := BakeryMachinesManager.MainPimentosDeposit_Value
            #Subdeposits
            WaterPreparedFor_B := BakeryMachinesManager.WaterPreparedFor_B
            FlourPreparedFor_B := BakeryMachinesManager.FlourPreparedFor_B
            WaterPreparedFor_CB := BakeryMachinesManager.WaterPreparedFor_CB
            FlourPreparedFor_CB := BakeryMachinesManager.FlourPreparedFor_CB
            CheesePreparedFor_CB := BakeryMachinesManager.CheesePreparedFor_CB
            WaterPreparedFor_PCS := BakeryMachinesManager.WaterPreparedFor_PCS
            FlourPreparedFor_PCS := BakeryMachinesManager.FlourPreparedFor_PCS
            CheesePreparedFor_PCS := BakeryMachinesManager.CheesePreparedFor_PCS
            PimentosPreparedFor_PCS := BakeryMachinesManager.PimentosPreparedFor_PCS
        #SaveNewData(Player,OldData,NewUpdatedData)#>

    #Areas
    SetLV_Area(ProfileData : player_data_table, NewLV_Area:int)<transacts>:player_data_table =
        NewData : player_data_table = player_data_table:
            MakePlayerDataTable<constructor>(ProfileData)
            LV_Area := NewLV_Area
        return NewData

    SetLast_Area(ProfileData : player_data_table, NewState:logic)<transacts>:player_data_table =
        NewData : player_data_table = player_data_table:
            MakePlayerDataTable<constructor>(ProfileData)
            LastAreaUnlocked := NewState
        return NewData
        

#Historic
    #Historic bread
    SetHistoric_ProductSent(ProfileData : player_data_table, B_Ammount:float, CB_Ammount:float, PCS_Ammount:float)<transacts>:player_data_table =
        NewData : player_data_table = player_data_table:
            MakePlayerDataTable<constructor>(ProfileData)
            Historic_BSent := B_Ammount
            Historic_CBSent := CB_Ammount
            Historic_PCSSent := PCS_Ammount
        return NewData

#MINIGAMES
    DecrementAllMiniGameCooldown(ProfileData : player_data_table, Amount : int)<transacts>:player_data_table =
        var AllMiniGamesCooldowns : []int = ProfileData.AllMiniGamesCooldowns
        for(Index:=0..AllMiniGamesCooldowns.Length-1, Value:=AllMiniGamesCooldowns[Index], set AllMiniGamesCooldowns[Index] = Value - Amount):
        NewData : player_data_table = player_data_table:
            MakePlayerDataTable<constructor>(ProfileData)
            AllMiniGamesCooldowns := AllMiniGamesCooldowns
        return NewData

    AddMiniGameCooldown(ProfileData : player_data_table, Index : int, CoolDownTime : int)<transacts>:player_data_table =
        var AllMiniGamesCooldowns : []int = ProfileData.AllMiniGamesCooldowns
        var AllMiniGamesCompleted : []logic = ProfileData.AllMiniGamesCompleted
        if(set AllMiniGamesCooldowns[Index] = CoolDownTime):
        if(set AllMiniGamesCompleted[Index] = true):
        NewData : player_data_table = player_data_table:
            MakePlayerDataTable<constructor>(ProfileData)
            AllMiniGamesCooldowns := AllMiniGamesCooldowns
            AllMiniGamesCompleted := AllMiniGamesCompleted
        return NewData

    #Sales efficency
    SetSalesEfficiency(ProfileData : player_data_table, NewLevels:[]int)<transacts>:player_data_table =
        NewData : player_data_table = player_data_table:
            MakePlayerDataTable<constructor>(ProfileData)
            AllLV_SalesEfficiency := NewLevels
        return NewData

    #CONSUME RESOURCERS
    SetSapphires(ProfileData : player_data_table, NewInventory_Sapphires:float)<transacts>:player_data_table =
        NewData : player_data_table = player_data_table:
            MakePlayerDataTable<constructor>(ProfileData)
            Inventory_Sapphires := NewInventory_Sapphires
        return NewData

    SetAgedSapphires(ProfileData : player_data_table, NewInventory_AgedSapphires:float)<transacts>:player_data_table =
        NewData : player_data_table = player_data_table:
            MakePlayerDataTable<constructor>(ProfileData)
            Inventory_AgedSapphires := NewInventory_AgedSapphires
        return NewData

        <#
    SetAgedSapphires(ProfileData : player_data_table, NewInventory_AgedSapphires:float)<transacts>:player_data_table =
        NewData : player_data_table = player_data_table:
            MakePlayerDataTable<constructor>(ProfileData)
            Inventory_AgedSapphires := NewInventory_AgedSapphires
        return NewData#>

    #PickaxePerks
    SetPickaxePerks(ProfileData : player_data_table, NewLevel:[]int)<transacts>:player_data_table =
        NewData : player_data_table = player_data_table:
            MakePlayerDataTable<constructor>(ProfileData)
            AllLV_PickaxePerks := NewLevel
        return NewData

    #Save Resourcers
    AddResourcers(ProfileData : player_data_table, Ammount : float, Type : int, Limit : float)<transacts>:tuple(player_data_table,logic) = 
        var WasOverLimit : logic = false
        if(Type = WATER):
            var NewAmmount : float = ProfileData.Inventory_Water +  Ammount
            if(Limit<NewAmmount):
                set WasOverLimit = true
                set NewAmmount = Limit
            NewData : player_data_table = player_data_table:
                MakePlayerDataTable<constructor>(ProfileData)
                Inventory_Water := NewAmmount
            return (NewData,WasOverLimit)
        else if (Type = FLOUR):
            var NewAmmount : float = ProfileData.Inventory_Flour +  Ammount
            if(Limit<NewAmmount):
                set WasOverLimit = true
                set NewAmmount = Limit
            NewData : player_data_table = player_data_table:
                MakePlayerDataTable<constructor>(ProfileData)
                Inventory_Flour := NewAmmount
            return (NewData,WasOverLimit)
        else if (Type = CHEESE):
            var NewAmmount : float = ProfileData.Inventory_Cheese +  Ammount
            if(Limit<NewAmmount):
                set WasOverLimit = true
                set NewAmmount = Limit
            NewData : player_data_table = player_data_table:
                MakePlayerDataTable<constructor>(ProfileData)
                Inventory_Cheese := NewAmmount
            return (NewData,WasOverLimit)
        else if (Type = PIMENTOS):
            var NewAmmount : float = ProfileData.Inventory_Pimentos +  Ammount
            if(Limit<NewAmmount):
                set WasOverLimit = true
                set NewAmmount = Limit
            NewData : player_data_table = player_data_table:
                MakePlayerDataTable<constructor>(ProfileData)
                Inventory_Pimentos := NewAmmount
            return (NewData,WasOverLimit)
        else if (Type = MUTANTMATERIAL):
            var NewAmmount : float = ProfileData.Inventory_MutantMaterial +  Ammount
            NewData : player_data_table = player_data_table:
                MakePlayerDataTable<constructor>(ProfileData)
                Inventory_MutantMaterial := NewAmmount
            return (NewData,WasOverLimit)
        else if (Type = COAL):
            var NewAmmount : float = ProfileData.Inventory_Coal +  Ammount
            NewData : player_data_table = player_data_table:
                MakePlayerDataTable<constructor>(ProfileData)
                Inventory_Coal := NewAmmount
            return (NewData,WasOverLimit)
        else if (Type = SAPPHIRES):
            var NewAmmount : float = ProfileData.Inventory_Sapphires +  Ammount
            NewData : player_data_table = player_data_table:
                MakePlayerDataTable<constructor>(ProfileData)
                Inventory_Sapphires := NewAmmount
            return (NewData,WasOverLimit)
        else if (Type = AGEDSAPPHIRES):
            var NewAmmount : float = ProfileData.Inventory_AgedSapphires +  Ammount
            NewData : player_data_table = player_data_table:
                MakePlayerDataTable<constructor>(ProfileData)
                Inventory_AgedSapphires := NewAmmount
            return (NewData,WasOverLimit)
        return (ProfileData,WasOverLimit)

    SetDepositResourcers(ProfileData : player_data_table, TupleResourcers : tuple(float,float,float,float))<transacts>:player_data_table = 
        NewData : player_data_table = player_data_table:
            MakePlayerDataTable<constructor>(ProfileData)
            MainWaterDeposit_Value := TupleResourcers(0)
            MainFlourDeposit_Value := TupleResourcers(1)
            MainCheeseDeposit_Value := TupleResourcers(2)
            MainPimentosDeposit_Value := TupleResourcers(3)
        return NewData

    AddDepositResourcers(ProfileData : player_data_table, TupleResourcers : tuple(float,float,float,float))<transacts>:player_data_table = 
        NewData : player_data_table = player_data_table:
            MakePlayerDataTable<constructor>(ProfileData)
            MainWaterDeposit_Value := ProfileData.MainWaterDeposit_Value + TupleResourcers(0)
            MainFlourDeposit_Value := ProfileData.MainFlourDeposit_Value + TupleResourcers(1)
            MainCheeseDeposit_Value := ProfileData.MainCheeseDeposit_Value + TupleResourcers(2)
            MainPimentosDeposit_Value := ProfileData.MainPimentosDeposit_Value + TupleResourcers(3)
        return NewData

    RemoveDepositResourcers(ProfileData : player_data_table, TupleResourcers : tuple(float,float,float,float))<transacts>:player_data_table = 
        NewData : player_data_table = player_data_table:
            MakePlayerDataTable<constructor>(ProfileData)
            MainWaterDeposit_Value := ProfileData.MainWaterDeposit_Value - TupleResourcers(0)
            MainFlourDeposit_Value := ProfileData.MainFlourDeposit_Value - TupleResourcers(1)
            MainCheeseDeposit_Value := ProfileData.MainCheeseDeposit_Value - TupleResourcers(2)
            MainPimentosDeposit_Value := ProfileData.MainPimentosDeposit_Value - TupleResourcers(3)
        return NewData

    RemoveProductsOnDrone(ProfileData : player_data_table)<transacts>:player_data_table = 
        NewData : player_data_table = player_data_table:
            MakePlayerDataTable<constructor>(ProfileData)
            B_On_Drone := 0.0
            CB_On_Drone := 0.0
            PCS_On_Drone := 0.0
        return NewData

    AddSubDepositResourcers(PlayerData : player_data_table, AllAmmounts : []float, ResourceType : int)<transacts>:player_data_table = 
        if(ResourceType = WATER, WaterAmmount_ForB_ToTake := AllAmmounts[0], WaterAmmount_ForCB_ToTake := AllAmmounts[1], WaterAmmount_ForPCS_ToTake := AllAmmounts[2]):
            NewData : player_data_table = player_data_table:
                MakePlayerDataTable<constructor>(PlayerData)
                WaterPreparedFor_B := PlayerData.WaterPreparedFor_B + WaterAmmount_ForB_ToTake
                WaterPreparedFor_CB := PlayerData.WaterPreparedFor_CB + WaterAmmount_ForCB_ToTake
                WaterPreparedFor_PCS := PlayerData.WaterPreparedFor_PCS + WaterAmmount_ForPCS_ToTake
            return NewData
        else if(ResourceType = FLOUR, FlourAmmount_ForB_ToTake := AllAmmounts[0], FlourAmmount_ForCB_ToTake := AllAmmounts[1], FlourAmmount_ForPCS_ToTake := AllAmmounts[2]):
            NewData : player_data_table = player_data_table:
                MakePlayerDataTable<constructor>(PlayerData)
                FlourPreparedFor_B := PlayerData.FlourPreparedFor_B + FlourAmmount_ForB_ToTake
                FlourPreparedFor_CB := PlayerData.FlourPreparedFor_CB + FlourAmmount_ForCB_ToTake
                FlourPreparedFor_PCS := PlayerData.FlourPreparedFor_PCS + FlourAmmount_ForPCS_ToTake
            return NewData
        else if(ResourceType = CHEESE, CheeseAmmount_ForCB_ToTake := AllAmmounts[0], CheeseAmmount_ForPCS_ToTake := AllAmmounts[1]):
            NewData : player_data_table = player_data_table:
                MakePlayerDataTable<constructor>(PlayerData)
                CheesePreparedFor_CB := PlayerData.CheesePreparedFor_CB + CheeseAmmount_ForCB_ToTake
                CheesePreparedFor_PCS := PlayerData.CheesePreparedFor_PCS + CheeseAmmount_ForPCS_ToTake
            return NewData
        else if(ResourceType = PIMENTOS, PimentosAmmount_ForPCS_ToTake := AllAmmounts[0]):
            NewData : player_data_table = player_data_table:
                MakePlayerDataTable<constructor>(PlayerData)
                PimentosPreparedFor_PCS := PlayerData.PimentosPreparedFor_PCS + PimentosAmmount_ForPCS_ToTake
            return NewData
        else:
            return PlayerData

        <#
    AddBakerResource_Inside(ProfileData : player_data_table, BakerType : int, ResourceType : int, Ammount : float)<transacts>:player_data_table = 
        #BB TYPE
        if(BakerType=BB_TYPE):
            if(ResourceType = WATER):
                NewData : player_data_table = player_data_table:
                    MakePlayerDataTable<constructor>(ProfileData)
                    BB_Water_Inside := ProfileData.BB_Water_Inside +  Ammount
                return NewData
            if(ResourceType = FLOUR):
                NewData : player_data_table = player_data_table:
                    MakePlayerDataTable<constructor>(ProfileData)
                    BB_Flour_Inside := ProfileData.BB_Flour_Inside +  Ammount
                return NewData
        #CBB TYPE
        if(BakerType=CBB_TYPE):
            if(ResourceType = WATER):
                NewData : player_data_table = player_data_table:
                    MakePlayerDataTable<constructor>(ProfileData)
                    CBB_Water_Inside := ProfileData.CBB_Water_Inside +  Ammount
                return NewData
            if(ResourceType = FLOUR):
                NewData : player_data_table = player_data_table:
                    MakePlayerDataTable<constructor>(ProfileData)
                    CBB_Flour_Inside := ProfileData.CBB_Flour_Inside +  Ammount
                return NewData
            if(ResourceType = CHEESE):
                NewData : player_data_table = player_data_table:
                    MakePlayerDataTable<constructor>(ProfileData)
                    CBB_Cheese_Inside := ProfileData.CBB_Cheese_Inside +  Ammount
                return NewData
        #PCSB TYPE
        if(BakerType=PCSB_TYPE):
            if(ResourceType = WATER):
                NewData : player_data_table = player_data_table:
                    MakePlayerDataTable<constructor>(ProfileData)
                    PCSB_Water_Inside := ProfileData.PCSB_Water_Inside +  Ammount
                return NewData
            if(ResourceType = FLOUR):
                NewData : player_data_table = player_data_table:
                    MakePlayerDataTable<constructor>(ProfileData)
                    PCSB_Flour_Inside := ProfileData.PCSB_Flour_Inside +  Ammount
                return NewData
            if(ResourceType = CHEESE):
                NewData : player_data_table = player_data_table:
                    MakePlayerDataTable<constructor>(ProfileData)
                    PCSB_Cheese_Inside := ProfileData.PCSB_Cheese_Inside +  Ammount
                return NewData
            if(ResourceType = PIMENTOS):
                NewData : player_data_table = player_data_table:
                    MakePlayerDataTable<constructor>(ProfileData)
                    PCSB_Pimentos_Inside := ProfileData.PCSB_Pimentos_Inside +  Ammount
                return NewData
        #PB TYPE
        if(ResourceType = MUTANTMATERIAL):
            NewData : player_data_table = player_data_table:
                MakePlayerDataTable<constructor>(ProfileData)
                PB_MutantMaterial_Inside := ProfileData.PB_MutantMaterial_Inside +  Ammount
            return NewData
        if(ResourceType = PIMENTOS):
            NewData : player_data_table = player_data_table:
                MakePlayerDataTable<constructor>(ProfileData)
                PB_Pimentos_Inside := ProfileData.PB_Pimentos_Inside +  Ammount
            return NewData

        return ProfileData
#>
<#
    WithdrawDepositResource(ProfileData : player_data_table, BakerType : int, AllAmmounts : []float)<transacts>:player_data_table = 
        #BB TYPE
        if(BakerType=BB_TYPE):
            if(AmountWater := AllAmmounts[WATER], AmountFlour := AllAmmounts[FLOUR]):
                NewData : player_data_table = player_data_table:
                    MakePlayerDataTable<constructor>(ProfileData)
                    WaterPreparedFor_B := ProfileData.WaterPreparedFor_B -  AmountWater
                    FlourPreparedFor_B := ProfileData.FlourPreparedFor_B -  AmountFlour
                return NewData
            #CBB TYPE
        else if(BakerType=CBB_TYPE):
            if(AmountWater := AllAmmounts[WATER], AmountFlour := AllAmmounts[FLOUR], AmountCheese:= AllAmmounts[CHEESE]):
                NewData : player_data_table = player_data_table:
                    MakePlayerDataTable<constructor>(ProfileData)
                    WaterPreparedFor_CB := ProfileData.WaterPreparedFor_CB -  AmountWater
                    FlourPreparedFor_CB := ProfileData.FlourPreparedFor_CB -  AmountFlour
                    CheesePreparedFor_CB := ProfileData.CheesePreparedFor_CB -  AmountCheese
            #PCSB TYPE
        else if(BakerType=PCSB_TYPE):
            if(AmountWater := AllAmmounts[WATER], AmountFlour := AllAmmounts[FLOUR], AmountCheese:= AllAmmounts[CHEESE], AmountPimentos:= AllAmmounts[PIMENTO]):
                NewData : player_data_table = player_data_table:
                    MakePlayerDataTable<constructor>(ProfileData)
                    WaterPreparedFor_PCS := ProfileData.WaterPreparedFor_PCS -  AmountWater
                    FlourPreparedFor_PCS := ProfileData.FlourPreparedFor_PCS -  AmountFlour
                    CheesePreparedFor_PCS := ProfileData.CheesePreparedFor_PCS - AmountCheese
                    PimentosPreparedFor_PCS := ProfileData.CheesePreparedFor_PCS - AmountPimentos
        else:
            return ProfileData
#>
<#
    WithdrawBakerResource_Inside(ProfileData : player_data_table, BakerType : int, ResourceType : int, Ammount : float)<transacts>:player_data_table = 
        #BB TYPE
        if(BakerType=BB_TYPE):
            if(ResourceType = WATER):
                NewData : player_data_table = player_data_table:
                    MakePlayerDataTable<constructor>(ProfileData)
                    BB_Water_Inside := ProfileData.BB_Water_Inside -  Ammount
                return NewData
            if(ResourceType = FLOUR):
                NewData : player_data_table = player_data_table:
                    MakePlayerDataTable<constructor>(ProfileData)
                    BB_Flour_Inside := ProfileData.BB_Flour_Inside -  Ammount
                return NewData
        #CBB TYPE
        if(BakerType=CBB_TYPE):
            if(ResourceType = WATER):
                NewData : player_data_table = player_data_table:
                    MakePlayerDataTable<constructor>(ProfileData)
                    CBB_Water_Inside := ProfileData.CBB_Water_Inside -  Ammount
                return NewData
            if(ResourceType = FLOUR):
                NewData : player_data_table = player_data_table:
                    MakePlayerDataTable<constructor>(ProfileData)
                    CBB_Flour_Inside := ProfileData.CBB_Flour_Inside -  Ammount
                return NewData
            if(ResourceType = CHEESE):
                NewData : player_data_table = player_data_table:
                    MakePlayerDataTable<constructor>(ProfileData)
                    CBB_Cheese_Inside := ProfileData.CBB_Cheese_Inside -  Ammount
                return NewData
        #PCSB TYPE
        if(BakerType=PCSB_TYPE):
            if(ResourceType = WATER):
                NewData : player_data_table = player_data_table:
                    MakePlayerDataTable<constructor>(ProfileData)
                    PCSB_Water_Inside := ProfileData.PCSB_Water_Inside -  Ammount
                return NewData
            if(ResourceType = FLOUR):
                NewData : player_data_table = player_data_table:
                    MakePlayerDataTable<constructor>(ProfileData)
                    PCSB_Flour_Inside := ProfileData.PCSB_Flour_Inside -  Ammount
                return NewData
            if(ResourceType = CHEESE):
                NewData : player_data_table = player_data_table:
                    MakePlayerDataTable<constructor>(ProfileData)
                    PCSB_Cheese_Inside := ProfileData.PCSB_Cheese_Inside -  Ammount
                return NewData
            if(ResourceType = PIMENTOS):
                NewData : player_data_table = player_data_table:
                    MakePlayerDataTable<constructor>(ProfileData)
                    PCSB_Pimentos_Inside := ProfileData.PCSB_Pimentos_Inside -  Ammount
                return NewData
        #PB TYPE
        if(ResourceType = MUTANTMATERIAL):
            NewData : player_data_table = player_data_table:
                MakePlayerDataTable<constructor>(ProfileData)
                PB_MutantMaterial_Inside := ProfileData.PB_MutantMaterial_Inside -  Ammount
            return NewData
        if(ResourceType = PIMENTOS):
            NewData : player_data_table = player_data_table:
                MakePlayerDataTable<constructor>(ProfileData)
                PB_Pimentos_Inside := ProfileData.PB_Pimentos_Inside -  Ammount
            return NewData

        return ProfileData#>

    ClearDroneAmmount(ProfileData : player_data_table)<transacts>:player_data_table = 
        NewData : player_data_table = player_data_table:
            MakePlayerDataTable<constructor>(ProfileData)
            B_On_Drone := 0.0
            CB_On_Drone := 0.0
            PCS_On_Drone := 0.0
        return NewData

    HasEnoughResource(ProfileData : player_data_table, Ammount : float, Type : int)<transacts>:tuple(int,float,float) = 
        var State : int = NOT_RES_STATE
        var ResToTakeFromMain : float = 0.0
        var ResToTakeFromSecondary : float = 0.0
        #WATER
        if(Type = WATER):
            if(ProfileData.Inventory_Water >= Ammount):
                set State = BACKPACK_RES_STATE
                set ResToTakeFromMain = 0.0
                set ResToTakeFromSecondary = 0.0
            else if(ProfileData.Inventory_Water + ProfileData.MainWaterDeposit_Value >= Ammount):
                set State = MAIN_RES_STATE
                set ResToTakeFromMain = Ammount - ProfileData.Inventory_Water
            else if(ProfileData.Inventory_Water + ProfileData.MainWaterDeposit_Value + (ProfileData.WaterPreparedFor_B+ProfileData.WaterPreparedFor_CB+ProfileData.WaterPreparedFor_PCS) >= Ammount):
                set State = SECOND_RES_STATE
                set ResToTakeFromSecondary = Ammount - ProfileData.Inventory_Water - ProfileData.MainWaterDeposit_Value
            else:
                set State = NOT_RES_STATE
        #FLOUR
        if(Type = FLOUR):
            if(ProfileData.Inventory_Flour >= Ammount):
                set State = BACKPACK_RES_STATE
                set ResToTakeFromMain = 0.0
                set ResToTakeFromSecondary = 0.0
            else if(ProfileData.Inventory_Flour + ProfileData.MainFlourDeposit_Value >= Ammount):
                set State = MAIN_RES_STATE
                set ResToTakeFromMain = Ammount - ProfileData.Inventory_Flour
            else if(ProfileData.Inventory_Flour + ProfileData.MainFlourDeposit_Value + (ProfileData.FlourPreparedFor_B+ProfileData.FlourPreparedFor_CB+ProfileData.FlourPreparedFor_PCS) >= Ammount):
                set State = SECOND_RES_STATE
                set ResToTakeFromSecondary = Ammount - ProfileData.Inventory_Flour - ProfileData.MainFlourDeposit_Value
            else:
                set State = NOT_RES_STATE
        #CHEESE
        if(Type = CHEESE):
            if(ProfileData.Inventory_Cheese >= Ammount):
                set State = BACKPACK_RES_STATE
                set ResToTakeFromMain = 0.0
                set ResToTakeFromSecondary = 0.0
            else if(ProfileData.Inventory_Cheese + ProfileData.MainCheeseDeposit_Value >= Ammount):
                set State = MAIN_RES_STATE
                set ResToTakeFromMain = Ammount - ProfileData.Inventory_Cheese
            else if(ProfileData.Inventory_Cheese + ProfileData.MainCheeseDeposit_Value + (ProfileData.CheesePreparedFor_CB+ProfileData.CheesePreparedFor_PCS) >= Ammount):
                set State = SECOND_RES_STATE
                set ResToTakeFromSecondary = Ammount - ProfileData.Inventory_Cheese - ProfileData.MainCheeseDeposit_Value
            else:
                set State = NOT_RES_STATE
        #PIMENTOS
        if(Type = PIMENTOS):
            if(ProfileData.Inventory_Pimentos >= Ammount):
                set State = BACKPACK_RES_STATE
                set ResToTakeFromMain = 0.0
                set ResToTakeFromSecondary = 0.0
            else if(ProfileData.Inventory_Pimentos + ProfileData.MainPimentosDeposit_Value >= Ammount):
                set State = MAIN_RES_STATE
                set ResToTakeFromMain = Ammount - ProfileData.Inventory_Pimentos
            else if(ProfileData.Inventory_Pimentos + ProfileData.MainPimentosDeposit_Value + ProfileData.PimentosPreparedFor_PCS >= Ammount):
                set State = SECOND_RES_STATE
                set ResToTakeFromSecondary = Ammount - ProfileData.Inventory_Pimentos - ProfileData.MainPimentosDeposit_Value
            else:
                set State = NOT_RES_STATE
        return (State,ResToTakeFromMain,ResToTakeFromSecondary)

    AddSuperEnergy(ProfileData : player_data_table, Amount : float)<transacts>:player_data_table = 
        NewData : player_data_table = player_data_table:
            MakePlayerDataTable<constructor>(ProfileData)
            SuperEnergy_Owned := ProfileData.SuperEnergy_Owned + Amount
        return NewData

    ConsumeSuperEnergy(ProfileData : player_data_table, Amount : float)<transacts>:player_data_table = 
        NewData : player_data_table = player_data_table:
            MakePlayerDataTable<constructor>(ProfileData)
            SuperEnergy_Owned := ProfileData.SuperEnergy_Owned - Amount
        return NewData

    WithdrawResourceFromSecondaryDeposit(ProfileData : player_data_table, Ammount : float, Type : int)<transacts> : player_data_table =
        if(Type=WATER): #WATER
            if(ProfileData.WaterPreparedFor_B >= Ammount):
                var NewAmount : float = ProfileData.WaterPreparedFor_B -  Ammount
                if(NewAmount<0.0):
                    set NewAmount=0.0
                NewData : player_data_table = player_data_table:
                    MakePlayerDataTable<constructor>(ProfileData)
                    WaterPreparedFor_B := NewAmount
                return NewData
            else if(ProfileData.WaterPreparedFor_B + ProfileData.WaterPreparedFor_CB >= Ammount):
                var NewAmount : float = ProfileData.WaterPreparedFor_CB - (Ammount - ProfileData.WaterPreparedFor_B)
                if(NewAmount<0.0):
                    set NewAmount=0.0
                NewData : player_data_table = player_data_table:
                    MakePlayerDataTable<constructor>(ProfileData)
                    WaterPreparedFor_B := 0.0
                    WaterPreparedFor_CB := NewAmount
                return NewData
            else if(ProfileData.WaterPreparedFor_B + ProfileData.WaterPreparedFor_CB + ProfileData.WaterPreparedFor_PCS >= Ammount):
                var NewAmount : float = ProfileData.WaterPreparedFor_PCS - (Ammount - ProfileData.WaterPreparedFor_B - ProfileData.WaterPreparedFor_CB)
                if(NewAmount<0.0):
                    set NewAmount=0.0
                NewData : player_data_table = player_data_table:
                    MakePlayerDataTable<constructor>(ProfileData)
                    WaterPreparedFor_B := 0.0
                    WaterPreparedFor_CB := 0.0
                    WaterPreparedFor_PCS := NewAmount
                return NewData
        else if(Type=FLOUR): #FLOUR
            if(ProfileData.FlourPreparedFor_B >= Ammount):
                var NewAmount : float = ProfileData.FlourPreparedFor_B -  Ammount
                if(NewAmount<0.0):
                    set NewAmount=0.0
                NewData : player_data_table = player_data_table:
                    MakePlayerDataTable<constructor>(ProfileData)
                    FlourPreparedFor_B := NewAmount
                return NewData
            else if(ProfileData.FlourPreparedFor_B + ProfileData.FlourPreparedFor_CB >= Ammount):
                var NewAmount : float = ProfileData.FlourPreparedFor_CB - (Ammount - ProfileData.FlourPreparedFor_B)
                if(NewAmount<0.0):
                    set NewAmount=0.0
                NewData : player_data_table = player_data_table:
                    MakePlayerDataTable<constructor>(ProfileData)
                    FlourPreparedFor_B := 0.0
                    FlourPreparedFor_CB := NewAmount
                return NewData
            else if(ProfileData.FlourPreparedFor_B + ProfileData.FlourPreparedFor_CB + ProfileData.FlourPreparedFor_PCS>= Ammount):
                var NewAmount : float = ProfileData.FlourPreparedFor_PCS - (Ammount - ProfileData.FlourPreparedFor_B - ProfileData.FlourPreparedFor_CB)
                if(NewAmount<0.0):
                    set NewAmount=0.0
                NewData : player_data_table = player_data_table:
                    MakePlayerDataTable<constructor>(ProfileData)
                    FlourPreparedFor_B := 0.0
                    FlourPreparedFor_CB := 0.0
                    FlourPreparedFor_PCS := NewAmount
                return NewData
        else if(Type=CHEESE):
            if(ProfileData.CheesePreparedFor_CB >= Ammount):
                var NewAmount : float = ProfileData.CheesePreparedFor_CB -  Ammount
                if(NewAmount<0.0):
                    set NewAmount=0.0
                NewData : player_data_table = player_data_table:
                    MakePlayerDataTable<constructor>(ProfileData)
                    CheesePreparedFor_CB := NewAmount
                return NewData
            else if(ProfileData.CheesePreparedFor_CB + ProfileData.CheesePreparedFor_PCS >= Ammount):
                var NewAmount : float = ProfileData.CheesePreparedFor_PCS - (Ammount - ProfileData.CheesePreparedFor_CB)
                if(NewAmount<0.0):
                    set NewAmount=0.0
                NewData : player_data_table = player_data_table:
                    MakePlayerDataTable<constructor>(ProfileData)
                    CheesePreparedFor_CB := 0.0
                    CheesePreparedFor_PCS := NewAmount
                return NewData
        else if(Type=PIMENTOS):
            if(ProfileData.PimentosPreparedFor_PCS >= Ammount):
                NewData : player_data_table = player_data_table:
                    MakePlayerDataTable<constructor>(ProfileData)
                    PimentosPreparedFor_PCS := ProfileData.PimentosPreparedFor_PCS - Ammount
                return NewData
        return ProfileData


    WithdrawResourceFromMainDeposit(ProfileData : player_data_table, Ammount : float, Type : int)<transacts>:player_data_table = 
        if(Type = WATER):
            var NewAmount : float = ProfileData.MainWaterDeposit_Value -  Ammount
            if(NewAmount<0.0):
                set NewAmount=0.0
            NewData : player_data_table = player_data_table:
                MakePlayerDataTable<constructor>(ProfileData)
                MainWaterDeposit_Value := NewAmount
            return NewData
        if(Type = FLOUR):
            var NewAmount : float = ProfileData.MainFlourDeposit_Value -  Ammount
            if(NewAmount<0.0):
                set NewAmount=0.0
            NewData : player_data_table = player_data_table:
                MakePlayerDataTable<constructor>(ProfileData)
                MainFlourDeposit_Value := NewAmount
            return NewData       
        if(Type = CHEESE):
            var NewAmount : float = ProfileData.MainCheeseDeposit_Value -  Ammount
            if(NewAmount<0.0):
                set NewAmount=0.0
            NewData : player_data_table = player_data_table:
                MakePlayerDataTable<constructor>(ProfileData)
                MainCheeseDeposit_Value := NewAmount
            return NewData   
        if(Type = PIMENTOS):
            var NewAmount : float = ProfileData.MainPimentosDeposit_Value -  Ammount
            if(NewAmount<0.0):
                set NewAmount=0.0
            NewData : player_data_table = player_data_table:
                MakePlayerDataTable<constructor>(ProfileData)
                MainPimentosDeposit_Value := NewAmount
            return NewData  
        return ProfileData 

    WithdrawResource(ProfileData : player_data_table, Ammount : float, Type : int)<transacts>:player_data_table = 
        if(Type = COAL):
            var NewAmount : float = ProfileData.Inventory_Coal -  Ammount
            if(NewAmount<0.0):
                set NewAmount=0.0
            NewData : player_data_table = player_data_table:
                MakePlayerDataTable<constructor>(ProfileData)
                Inventory_Coal := NewAmount
            return NewData
        if(Type = WATER):
            var NewAmount : float = ProfileData.Inventory_Water -  Ammount
            if(NewAmount<0.0):
                set NewAmount=0.0
            NewData : player_data_table = player_data_table:
                MakePlayerDataTable<constructor>(ProfileData)
                Inventory_Water := NewAmount
            return NewData
        if(Type = FLOUR):
            var NewAmount : float = ProfileData.Inventory_Flour -  Ammount
            if(NewAmount<0.0):
                set NewAmount=0.0
            NewData : player_data_table = player_data_table:
                MakePlayerDataTable<constructor>(ProfileData)
                Inventory_Flour := NewAmount
            return NewData       
        if(Type = CHEESE):
            var NewAmount : float = ProfileData.Inventory_Cheese -  Ammount
            if(NewAmount<0.0):
                set NewAmount=0.0
            NewData : player_data_table = player_data_table:
                MakePlayerDataTable<constructor>(ProfileData)
                Inventory_Cheese := NewAmount
            return NewData   
        if(Type = PIMENTOS):
            var NewAmount : float = ProfileData.Inventory_Pimentos -  Ammount
            if(NewAmount<0.0):
                set NewAmount=0.0
            NewData : player_data_table = player_data_table:
                MakePlayerDataTable<constructor>(ProfileData)
                Inventory_Pimentos := NewAmount
            return NewData   
        if(Type = MUTANTMATERIAL):
            var NewAmount : float = ProfileData.Inventory_MutantMaterial -  Ammount
            if(NewAmount<0.0):
                set NewAmount=0.0
            NewData : player_data_table = player_data_table:
                MakePlayerDataTable<constructor>(ProfileData)
                Inventory_MutantMaterial := NewAmount
            return NewData   
        if(Type = COAL):
            var NewAmount : float = ProfileData.Inventory_Coal -  Ammount
            if(NewAmount<0.0):
                set NewAmount=0.0
            NewData : player_data_table = player_data_table:
                MakePlayerDataTable<constructor>(ProfileData)
                Inventory_Coal := NewAmount
            return NewData   
        if(Type = AGEDSAPPHIRES):
            var NewAmount : float = ProfileData.Inventory_AgedSapphires -  Ammount
            if(NewAmount<0.0):
                set NewAmount=0.0
            NewData : player_data_table = player_data_table:
                MakePlayerDataTable<constructor>(ProfileData)
                Inventory_AgedSapphires := NewAmount
            return NewData   
        if(Type = SAPPHIRES):
            var NewAmount : float = ProfileData.Inventory_Sapphires -  Ammount
            if(NewAmount<0.0):
                set NewAmount=0.0
            NewData : player_data_table = player_data_table:
                MakePlayerDataTable<constructor>(ProfileData)
                Inventory_Sapphires := NewAmount
            return NewData
        return ProfileData

    SetProductToDrone(ProfileData : player_data_table, Ammount : float, Type : int) : player_data_table =
        if(Type = B_TYPE):
            NewData : player_data_table = player_data_table:
                MakePlayerDataTable<constructor>(ProfileData)
                B_On_Drone := Ammount
            return NewData
        if(Type = CB_TYPE):
            NewData : player_data_table = player_data_table:
                MakePlayerDataTable<constructor>(ProfileData)
                CB_On_Drone := Ammount
            return NewData       
        if(Type = PCS_TYPE):
            NewData : player_data_table = player_data_table:
                MakePlayerDataTable<constructor>(ProfileData)
                PCS_On_Drone := Ammount
            return NewData   
        else:
            return ProfileData

    SetMusicState(ProfileData : player_data_table, NewState : logic)<transacts>:player_data_table =
        NewData : player_data_table = player_data_table:
            MakePlayerDataTable<constructor>(ProfileData)
            IsMusicOn := NewState
        return NewData

    SetTrackerState(ProfileData : player_data_table, NewState : logic)<transacts>:player_data_table =
        NewData : player_data_table = player_data_table:
            MakePlayerDataTable<constructor>(ProfileData)
            IsQuestTrackerOn := NewState
        return NewData

    EvolveDungeon(ProfileData : player_data_table, DungeonIndex : int)<transacts>:player_data_table =
        var CurrentEvolvedDungeons : []int = ProfileData.AllEvolvedDungeons
        if(NewValue:=CurrentEvolvedDungeons[DungeonIndex]+1, set CurrentEvolvedDungeons[DungeonIndex]=NewValue):
            NewData : player_data_table = player_data_table:
                MakePlayerDataTable<constructor>(ProfileData)
                AllEvolvedDungeons := CurrentEvolvedDungeons
            return NewData
        else:
            return ProfileData

    RegisterSecret(ProfileData : player_data_table, ScrollIndex : int)<transacts>:tuple(player_data_table,logic) =
        var CurrentSecrets : []logic = ProfileData.AllSecrets_Granted
        if(IsAlreadyregistered:=CurrentSecrets[ScrollIndex]):
            if(not IsAlreadyregistered?):
                if(set CurrentSecrets[ScrollIndex]=true):
                    NewData : player_data_table = player_data_table:
                        MakePlayerDataTable<constructor>(ProfileData)
                        AllSecrets_Granted := CurrentSecrets
                    return (NewData,true)
                else:
                    return (ProfileData,false)
            else:
                return (ProfileData,false)
        else:
            return (ProfileData,false)

    WithdrawFromDrone(ProfileData : player_data_table, Ammount : float, Type : int)<transacts>:player_data_table = 
        if(Type = B_TYPE):
            NewData : player_data_table = player_data_table:
                MakePlayerDataTable<constructor>(ProfileData)
                B_On_Drone := ProfileData.B_On_Drone -  Ammount
            return NewData
        if(Type = CB_TYPE):
            NewData : player_data_table = player_data_table:
                MakePlayerDataTable<constructor>(ProfileData)
                CB_On_Drone := ProfileData.CB_On_Drone -  Ammount
            return NewData       
        if(Type = PCS_TYPE):
            NewData : player_data_table = player_data_table:
                MakePlayerDataTable<constructor>(ProfileData)
                PCS_On_Drone := ProfileData.PCS_On_Drone -  Ammount
            return NewData   
        else:
            return ProfileData


#PICKAXE
    SetTLV_PickaxeStrength(ProfileData : player_data_table, NewLevels :[]int)<transacts>:player_data_table =
        NewData : player_data_table = player_data_table:
            MakePlayerDataTable<constructor>(ProfileData)
            TLV_PickaxeStrength := NewLevels
        return NewData

    SetTLV_PickaxeEfficiency(ProfileData : player_data_table, NewLevels : []int)<transacts>:player_data_table =
        NewData : player_data_table = player_data_table:
            MakePlayerDataTable<constructor>(ProfileData)
            LV_PickaxeEfficiency := NewLevels
        return NewData

#BAKERS
    #PimientoCheeseBlender
    SetTLV_BBaker(ProfileData : player_data_table, NewLevels :[]int)<transacts>:player_data_table =
        NewData : player_data_table = player_data_table:
            MakePlayerDataTable<constructor>(ProfileData)
            TLV_BB := NewLevels
        return NewData

    SetTLV_CBBaker(ProfileData : player_data_table, NewLevels :[]int)<transacts>:player_data_table =
        NewData : player_data_table = player_data_table:
            MakePlayerDataTable<constructor>(ProfileData)
            TLV_CBB := NewLevels
        return NewData

    SetTLV_PCSBaker(ProfileData : player_data_table, NewLevels :[]int)<transacts>:player_data_table =
        NewData : player_data_table = player_data_table:
            MakePlayerDataTable<constructor>(ProfileData)
            TLV_PCSB := NewLevels
        return NewData

    SetPCSB_Efficiency(ProfileData : player_data_table, NewLevels : []int)<transacts>:player_data_table =
        NewData : player_data_table = player_data_table:
            MakePlayerDataTable<constructor>(ProfileData)
            EfficiencyLV_PCSB := NewLevels
        return NewData

    SetCBB_Efficiency(ProfileData : player_data_table, NewLevels : []int)<transacts>:player_data_table =
        NewData : player_data_table = player_data_table:
            MakePlayerDataTable<constructor>(ProfileData)
            EfficiencyLV_CBB := NewLevels
        return NewData

    SetBB_Efficiency(ProfileData : player_data_table, NewLevels : []int)<transacts>:player_data_table =
        NewData : player_data_table = player_data_table:
            MakePlayerDataTable<constructor>(ProfileData)
            EfficiencyLV_BB := NewLevels
        return NewData

    SetPCSB_Speed(ProfileData : player_data_table, NewLevel : int)<transacts>:player_data_table =
        NewData : player_data_table = player_data_table:
            MakePlayerDataTable<constructor>(ProfileData)
            LV_PCSBSpeed := NewLevel
        return NewData

    SetCBB_Speed(ProfileData : player_data_table, NewLevel : int)<transacts>:player_data_table =
        NewData : player_data_table = player_data_table:
            MakePlayerDataTable<constructor>(ProfileData)
            LV_CBBSpeed := NewLevel
        return NewData

    SetBB_Speed(ProfileData : player_data_table, NewLevel : int)<transacts>:player_data_table =
        NewData : player_data_table = player_data_table:
            MakePlayerDataTable<constructor>(ProfileData)
            LV_BBSpeed := NewLevel
        return NewData

    #MainDeposits
    SetLV_MainDeposits(ProfileData : player_data_table, NewLevels :[]int)<transacts>:player_data_table =
        NewData : player_data_table = player_data_table:
            MakePlayerDataTable<constructor>(ProfileData)
            AllLV_MainDeposits := NewLevels
        return NewData

    #Pasive Resourcers
    SetLV_PasiveResourcers(ProfileData : player_data_table, NewLevels :[]int)<transacts>:player_data_table =
        NewData : player_data_table = player_data_table:
            MakePlayerDataTable<constructor>(ProfileData)
            AllLV_PasiveResourcers := NewLevels
        return NewData

    #DroneSpeed
    SetDroneSpeed(ProfileData : player_data_table, NewLevel :int)<transacts>:player_data_table =
        NewData : player_data_table = player_data_table:
            MakePlayerDataTable<constructor>(ProfileData)
            LV_DroneSpeed := NewLevel
        return NewData

    #Backpack
    SetAllLV_BackpackLimits(ProfileData : player_data_table, NewLevels : []int)<transacts>:player_data_table =
        NewData : player_data_table = player_data_table:
            MakePlayerDataTable<constructor>(ProfileData)
            AllLV_BackpackLimits := NewLevels
        return NewData

#PETS STATS
    SavePets(ProfileData : player_data_table, AllPetsData:[]pet_data_table) <transacts>:player_data_table =
        NewData : player_data_table = player_data_table:
            MakePlayerDataTable<constructor>(ProfileData)
            AllPets := AllPetsData

    SetHexaGamePosition(PetData : pet_data_table, HexaGamePositionX : int, HexaGamePositionY : int) <transacts>:pet_data_table =
        NewData : pet_data_table = pet_data_table :
            MakePetDataTable<constructor>(PetData)
            HexaGamePositionX := HexaGamePositionX
            HexaGamePositionY := HexaGamePositionY
        return NewData   

    SetChallengePointer(ProfileData : player_data_table, HexaGame_ChallengePointer : int) <transacts>:player_data_table =
        NewData : player_data_table = player_data_table:
            MakePlayerDataTable<constructor>(ProfileData)
            HexaGame_ChallengePointer := HexaGame_ChallengePointer

    SetPetPerk(ProfileData : player_data_table, NewLV:int, PetStatType : int) <transacts>:player_data_table =
        if(PetStatType=STAT_TYPE_PET_PERK_MAXACTIVE):
            NewData : player_data_table = player_data_table:
                MakePlayerDataTable<constructor>(ProfileData)
                MaxActive_PetSlots := NewLV
            return NewData
        else if(PetStatType=STAT_TYPE_PET_PERK_MAXSLOTS):
            NewData : player_data_table = player_data_table:
                MakePlayerDataTable<constructor>(ProfileData)
                MaxNoActive_PetSlots := NewLV
            return NewData
        else if(PetStatType=STAT_TYPE_PET_PERK_TIER):  
            NewData : player_data_table = player_data_table:
                MakePlayerDataTable<constructor>(ProfileData)
                PetTier := NewLV
            return NewData
        return ProfileData

    SetPetPackA_LV(PetData:pet_data_table, LV:int) <transacts>:pet_data_table =
        NewData : pet_data_table = pet_data_table :
            MakePetDataTable<constructor>(PetData)
            LVPet_PackageA := LV
        return NewData   
        
    SetPetPackB_LV(PetData:pet_data_table, LV:int)<transacts>:pet_data_table =
        NewData : pet_data_table = pet_data_table :
            MakePetDataTable<constructor>(PetData)
            LVPet_PackageB := LV
        return NewData 

    ImprovePetPurity(PetData:pet_data_table, Ammount:int) <transacts>:pet_data_table =
        NewData : pet_data_table = pet_data_table :
            MakePetDataTable<constructor>(PetData)
            Purity := PetData.Purity + Ammount
        return NewData

    CreatePet(InPetData:pet_data_table,InPetType:int,InPurity:int,InLVPet_PackageA:int,InLVPet_PackageB:int)<transacts>:pet_data_table =
        NewData : pet_data_table = pet_data_table :
            MakePetDataTable<constructor>(InPetData)
            Type := InPetType
            Purity := InPurity
            LVPet_PackageA:=InLVPet_PackageA
            LVPet_PackageB:=InLVPet_PackageB
        return NewData

    SetEquippedPet(PetData:pet_data_table, Equiped:logic) <transacts>:pet_data_table =
        NewData : pet_data_table = pet_data_table :
            MakePetDataTable<constructor>(PetData)
            IsEquipped := Equiped
        return NewData

    AddNewPet(ProfileData : player_data_table, PetData:pet_data_table)<transacts>:player_data_table =
        NewPets := ProfileData.AllPets + array{PetData}
        NewData : player_data_table = player_data_table:
            MakePlayerDataTable<constructor>(ProfileData)
            AllPets := NewPets
        return NewData

    CalculePetFusion<public>(Pet0 : pet_data_table, Pet1 : pet_data_table, MaxCurrentPetTier : int)<suspends> : tuple(pet_data_table,logic) =
        var IsTierLimited : logic = false
        FusionPet : pet_data_table = pet_data_table{}
        if(Pet0.Type=Pet1.Type):
            TupleInvestedPet0:=CalculePetTotalAmountsInvested(Pet0)
            Sleep(0.04)
            TupleInvestedPet1:=CalculePetTotalAmountsInvested(Pet1)
            Sleep(0.04)
            var TotalAmountForPurity : float = (TupleInvestedPet0(0)+TupleInvestedPet1(0))*FusionEfficiency
            var TotalAmountForPackageA : float = (TupleInvestedPet0(1)+TupleInvestedPet1(1))*FusionEfficiency
            var TotalAmountForPackageB : float = (TupleInvestedPet0(2)+TupleInvestedPet1(2))*FusionEfficiency
    
            var TotalLVPurityReached : int = 1
            var TotalLVPackageAReached : int = 1
            var TotalLVPackageBReached : int = 1
    
            #Purity reached
            loop:
                NextCost:=MainFormula_PetPurityCost(TotalLVPurityReached)
                if(TotalAmountForPurity>=NextCost, TotalLVPurityReached < PetTotalMaxPurity):
                    set TotalLVPurityReached = TotalLVPurityReached + 1
                else:
                    #Print("Purity combined reached : {TotalLVPurityReached}")
                    break

            if(MaxCurrentPetTier<2, TotalLVPurityReached>99):
                set IsTierLimited = true
                set TotalLVPurityReached = 99
            else if(MaxCurrentPetTier<3, TotalLVPurityReached>199):
                set IsTierLimited = true
                set TotalLVPurityReached = 199
           
            Sleep(0.04)
            #PackageA reached
            loop:
                NextCost:=MainFormula_PackageACost(TotalLVPackageAReached)
                if(TotalAmountForPackageA>=NextCost, TotalLVPackageAReached < MaxLV_Pet_PackageA):
                    set TotalLVPackageAReached = TotalLVPackageAReached + 1
                    set TotalAmountForPackageA = TotalAmountForPackageA - NextCost
                else:
                    break
            Sleep(0.04)
            #PackageB reached
            loop:
                NextCost:=MainFormula_PackageBCost(TotalLVPackageBReached)
                if(TotalAmountForPackageB>=NextCost, TotalLVPackageBReached < MaxLV_Pet_PackageB):
                    set TotalLVPackageBReached = TotalLVPackageBReached + 1
                    set TotalAmountForPackageB = TotalAmountForPackageB - NextCost
                else:
                    break
            NewFusionPet:=CreatePet(FusionPet,Pet0.Type,TotalLVPurityReached,TotalLVPackageAReached,TotalLVPackageBReached)
            return (NewFusionPet,IsTierLimited)
        return (FusionPet,IsTierLimited)

    #HEXAGAME
    ImproveHexaGamePerk(ProfileData : player_data_table, Perk : int) <transacts>:player_data_table =
        if(Perk = HEXAGAMEPERK_DELAY):
            NewData : player_data_table = player_data_table:
                MakePlayerDataTable<constructor>(ProfileData)
                HexaGame_LVDelay := ProfileData.HexaGame_LVDelay + 1
            return NewData
        else if(Perk = HEXAGAMEPERK_ACCUMULATED):
            NewData : player_data_table = player_data_table:
                MakePlayerDataTable<constructor>(ProfileData)
                HexaGame_LVAccumulated := ProfileData.HexaGame_LVAccumulated + 1
            return NewData
        return ProfileData

    AddAccumulatedHexaToken(ProfileData : player_data_table, ResetTime : float)<transacts>:player_data_table =
        if(ResetTime<>-1):
            NewData : player_data_table = player_data_table:
                MakePlayerDataTable<constructor>(ProfileData)
                HexaGame_Accumulated := ProfileData.HexaGame_Accumulated + 1
                HexaGame_DelayTime:= ResetTime
            return NewData
        else:
            NewData : player_data_table = player_data_table:
                MakePlayerDataTable<constructor>(ProfileData)
                HexaGame_Accumulated := ProfileData.HexaGame_Accumulated + 1
            return NewData

    AddAccumulatedDungeonToken(ProfileData : player_data_table, ResetTime : float)<transacts>:player_data_table =
        if(ResetTime<>-1):
            NewData : player_data_table = player_data_table:
                MakePlayerDataTable<constructor>(ProfileData)
                Dungeon_Accumulated := ProfileData.Dungeon_Accumulated + 1
                Dungeon_DelayTime:= ResetTime
            return NewData
        else:
            NewData : player_data_table = player_data_table:
                MakePlayerDataTable<constructor>(ProfileData)
                Dungeon_Accumulated := ProfileData.Dungeon_Accumulated + 1
            return NewData


    RemoveAccumulatedHexaChallengeToken(ProfileData : player_data_table)<transacts>:player_data_table =
        NewData : player_data_table = player_data_table:
            MakePlayerDataTable<constructor>(ProfileData)
            HexaGame_Accumulated := ProfileData.HexaGame_Accumulated -1
        return NewData

    RegisterHexaChallengeCompleted(ProfileData : player_data_table, ChallengeLV : int) <transacts>:tuple(player_data_table,logic) =
        var IsFirstTimeCompleted : logic = true
        var NewHexaGame_AllChallengesCompleted : []int = ProfileData.HexaGame_AllChallengesCompleted
        for(ChallengeCompletedAlready : ProfileData.HexaGame_AllChallengesCompleted, ChallengeCompletedAlready = ChallengeLV):
            set IsFirstTimeCompleted = false
        if(IsFirstTimeCompleted?):
            set NewHexaGame_AllChallengesCompleted = NewHexaGame_AllChallengesCompleted + array{ChallengeLV}
        NewData : player_data_table = player_data_table:
            MakePlayerDataTable<constructor>(ProfileData)
            HexaGame_AllChallengesCompleted := NewHexaGame_AllChallengesCompleted
        return (NewData,IsFirstTimeCompleted)

    SetTotemHexaChallenge(ProfileData : player_data_table, ChallengeTotemIndex : int) <transacts>: tuple(player_data_table,logic) =
        var AllHexaGame_TotemsUnlocked : []logic = ProfileData.AllHexaGame_TotemsUnlocked
        var NewDataTable : player_data_table = ProfileData
        var IsNewRegister : logic = false
        if(IsUnlocked:=AllHexaGame_TotemsUnlocked[ChallengeTotemIndex]):
            if(IsUnlocked?):
                set IsNewRegister = false
            else if(set AllHexaGame_TotemsUnlocked[ChallengeTotemIndex] = true):
                set IsNewRegister = true
        if(IsNewRegister?):
            NewData : player_data_table = player_data_table:
                MakePlayerDataTable<constructor>(ProfileData)
                AllHexaGame_TotemsUnlocked := AllHexaGame_TotemsUnlocked
            set NewDataTable = NewData
        return (NewDataTable,IsNewRegister)

    DecrementDungeon_Delay(ProfileData : player_data_table, Amount : float) <transacts>:player_data_table =
        var NewDungeon_DelayTime : float = ProfileData.Dungeon_DelayTime - Amount
        if(0.0 > NewDungeon_DelayTime):
            set NewDungeon_DelayTime = 0.0
        NewData : player_data_table = player_data_table:
            MakePlayerDataTable<constructor>(ProfileData)
            Dungeon_DelayTime := NewDungeon_DelayTime
        return NewData

    DecrementHexaGame_Delay(ProfileData : player_data_table, Amount : float) <transacts>:player_data_table =
        var NewHexaGame_DelayTime : float = ProfileData.HexaGame_DelayTime - Amount
        if(0.0 > NewHexaGame_DelayTime):
            set NewHexaGame_DelayTime = 0.0
        NewData : player_data_table = player_data_table:
            MakePlayerDataTable<constructor>(ProfileData)
            HexaGame_DelayTime := NewHexaGame_DelayTime
        return NewData

    SetTutorialState(ProfileData : player_data_table, NewValue : logic)<transacts>:player_data_table =
        NewData : player_data_table = player_data_table:
            MakePlayerDataTable<constructor>(ProfileData)
            IsTutorialCompleted := NewValue
        return NewData

    DecrementAllDelays(ProfileData : player_data_table, Amount : float) <transacts>:player_data_table =
        var NewDungeon_DelayTime : float = ProfileData.Dungeon_DelayTime
        if(ProfileData.Dungeon_Accumulated<ProfileData.Dungeon_LVAccumulated, ProfileData.NumQuestsCompleted >= TutorialDungeon_Index):
            set NewDungeon_DelayTime = ProfileData.Dungeon_DelayTime - Amount

        var NewHexaGame_DelayTime : float = ProfileData.HexaGame_DelayTime
        if(ProfileData.HexaGame_Accumulated<ProfileData.HexaGame_LVAccumulated+HexaGame_MinToken_Accumulated_MinusOne, ProfileData.NumQuestsCompleted >= TutorialHexaGame_Index):
            set NewHexaGame_DelayTime = ProfileData.HexaGame_DelayTime - Amount

        var NewQuest_DelayTime : float = ProfileData.Quest_DelayTime
        if(ProfileData.NumQuestsCompleted > Quest0_Index):
            set NewQuest_DelayTime = ProfileData.Quest_DelayTime - Amount

        var NewOrder_DelayTime : float = ProfileData.Order_DelayTime
        if(ProfileData.NumQuestsCompleted > TutorialOrder_Index):
            set NewOrder_DelayTime = ProfileData.Order_DelayTime - Amount

        if(0.0 > NewDungeon_DelayTime):
            set NewDungeon_DelayTime = 0.0
        if(0.0 > NewHexaGame_DelayTime):
            set NewHexaGame_DelayTime = 0.0
        if(0.0 > NewQuest_DelayTime):
            set NewQuest_DelayTime = 0.0
        if(0.0 > NewOrder_DelayTime):
            set NewOrder_DelayTime = 0.0

        NewData : player_data_table = player_data_table:
            MakePlayerDataTable<constructor>(ProfileData)
            Dungeon_DelayTime := NewDungeon_DelayTime
            HexaGame_DelayTime := NewHexaGame_DelayTime
            Quest_DelayTime := NewQuest_DelayTime 
            Order_DelayTime := NewOrder_DelayTime
        return NewData

    #DUNGEONS
    ImproveDungeonPerk(ProfileData : player_data_table, Perk : int) <transacts>:player_data_table =
        if(Perk = DUNGEONPERK_DELAY):
            NewData : player_data_table = player_data_table:
                MakePlayerDataTable<constructor>(ProfileData)
                Dungeon_LVDelay := ProfileData.Dungeon_LVDelay + 1
            return NewData
        else if(Perk = DUNGEONPERK_ACCUMULATED):
            NewData : player_data_table = player_data_table:
                MakePlayerDataTable<constructor>(ProfileData)
                Dungeon_LVAccumulated := ProfileData.Dungeon_LVAccumulated + 1
            return NewData
        return ProfileData

    AddDungeonAccumulated(ProfileData : player_data_table) <transacts>:player_data_table =
        NewDungeon_Accumulated:=ProfileData.Dungeon_Accumulated+1
        NewData : player_data_table = player_data_table:
            MakePlayerDataTable<constructor>(ProfileData)
            Dungeon_Accumulated := NewDungeon_Accumulated
        return NewData    

    ConsumeDungeonAccumulated(ProfileData : player_data_table) <transacts>:player_data_table =
        NewDungeon_Accumulated:=ProfileData.Dungeon_Accumulated-1
        NewData : player_data_table = player_data_table:
            MakePlayerDataTable<constructor>(ProfileData)
            Dungeon_Accumulated := NewDungeon_Accumulated
        return NewData     

#ORDERS AND QUESTS
    CreateNewOrder(ProfileData : player_data_table, RequestedType : int, TotalRequested : float, RewardType : int, TotalReward : float) <transacts>:player_data_table =
        NewDungeon_Accumulated:=ProfileData.Dungeon_Accumulated-1
        NewData : player_data_table = player_data_table:
            MakePlayerDataTable<constructor>(ProfileData)
            OrderAssigned_ProgressAmount := 0.0
            OrderAssigned_ProductAmountRequested := TotalRequested
            OrderAssigned_ProductRequested := RequestedType
            OrderAssigned_Reward := TotalReward
            OrderAssigned_RewardType := RewardType
        return NewData   
        
    SetOrder_Progress(ProfileData : player_data_table, NewAmount : float) <transacts>:player_data_table =
        NewData : player_data_table = player_data_table:
            MakePlayerDataTable<constructor>(ProfileData)
            OrderAssigned_ProgressAmount := NewAmount


    CreateOrder_Delay(ProfileData : player_data_table, DelayTime : float) <transacts>:player_data_table =
        NewData : player_data_table = player_data_table:
            MakePlayerDataTable<constructor>(ProfileData)
            Order_DelayTime := DelayTime
            OrderAssigned_ProductRequested := -99
            OrderAssigned_ProgressAmount := -99.0
            OrderAssigned_Reward := -99.0
            OrderAssigned_RewardType := -99

    CreateNewQuest(ProfileData : player_data_table, NumQuest : int, ObjetiveType : int,RewardType : int, TotalReward : float) <transacts>:player_data_table =
        NewData : player_data_table = player_data_table:
            MakePlayerDataTable<constructor>(ProfileData)
            QuestAssigned_NumQuest := NumQuest
            QuestAssigned_ObjetiveType := ObjetiveType
            QuestAssigned_ProgressAmount := 0.0
            QuestAssigned_Reward := TotalReward
            QuestAssigned_RewardType := RewardType
        return NewData 

    SetQuest_Progress(ProfileData : player_data_table, NewAmount : float) <transacts>:player_data_table =
        NewData : player_data_table = player_data_table:
            MakePlayerDataTable<constructor>(ProfileData)
            QuestAssigned_ProgressAmount := NewAmount

    CreateQuest_Delay(ProfileData : player_data_table, DelayTime : float) <transacts>:player_data_table =
        NewData : player_data_table = player_data_table:
            MakePlayerDataTable<constructor>(ProfileData)
            Quest_DelayTime := DelayTime
            QuestAssigned_NumQuest := -99
            QuestAssigned_ProgressAmount := -99.0
            QuestAssigned_Reward := -99.0
            QuestAssigned_RewardType := -99

#WEAPONS
    #AMMO
    SetAmmoArrow(ProfileData : player_data_table, NewAmount : int) <transacts>:player_data_table =
        NewData : player_data_table = player_data_table:
            MakePlayerDataTable<constructor>(ProfileData)
            Ammo_Arrows := NewAmount
        return NewData

    ImproveWeaponStat(ProfileData : player_data_table, Perk : int, WeaponType : int) <transacts>:player_data_table =
        if(WeaponType=SWORD_TYPE):
            if(Perk = WEAPONSTAT_TYPE_EVOLUTION):
                NewData : player_data_table = player_data_table:
                    MakePlayerDataTable<constructor>(ProfileData)
                    Tier_Sword := ProfileData.Tier_Sword + 1
                return NewData
            else if(Perk = WEAPONSTAT_TYPE_DAMAGE):
                NewData : player_data_table = player_data_table:
                    MakePlayerDataTable<constructor>(ProfileData)
                    LV_SwordDamage := ProfileData.LV_SwordDamage + 1
                return NewData
            else if(Perk = WEAPONSTAT_TYPE_CRITICALPROB):
                NewData : player_data_table = player_data_table:
                    MakePlayerDataTable<constructor>(ProfileData)
                    LV_SwordCriticalProb := ProfileData.LV_SwordCriticalProb + 1
                return NewData
            else if(Perk = WEAPONSTAT_TYPE_CRITICALDAMAGE):
                NewData : player_data_table = player_data_table:
                    MakePlayerDataTable<constructor>(ProfileData)
                    LV_SwordCriticalDamage := ProfileData.LV_SwordCriticalDamage + 1
                return NewData
            else if(Perk = WEAPONSTAT_TYPE_REWARD):
                NewData : player_data_table = player_data_table:
                    MakePlayerDataTable<constructor>(ProfileData)
                    LV_SwordReward := ProfileData.LV_SwordReward + 1
                return NewData
        else:
            if(Perk = WEAPONSTAT_TYPE_EVOLUTION):
                NewData : player_data_table = player_data_table:
                    MakePlayerDataTable<constructor>(ProfileData)
                    Tier_Bow := ProfileData.Tier_Bow + 1
                return NewData
            else if(Perk = WEAPONSTAT_TYPE_DAMAGE):
                NewData : player_data_table = player_data_table:
                    MakePlayerDataTable<constructor>(ProfileData)
                    LV_BowDamage := ProfileData.LV_BowDamage + 1
                return NewData
            else if(Perk = WEAPONSTAT_TYPE_CRITICALPROB):
                NewData : player_data_table = player_data_table:
                    MakePlayerDataTable<constructor>(ProfileData)
                    LV_BowCriticalProb := ProfileData.LV_BowCriticalProb + 1
                return NewData
            else if(Perk = WEAPONSTAT_TYPE_CRITICALDAMAGE):
                NewData : player_data_table = player_data_table:
                    MakePlayerDataTable<constructor>(ProfileData)
                    LV_BowCriticalDamage := ProfileData.LV_BowCriticalDamage + 1
                return NewData
            else if(Perk = WEAPONSTAT_TYPE_REWARD):
                NewData : player_data_table = player_data_table:
                    MakePlayerDataTable<constructor>(ProfileData)
                    LV_BowReward := ProfileData.LV_BowReward + 1
                return NewData
        return ProfileData

    #TUTORIALS / QUESTS
    SetNumQuestsCompleted (ProfileData : player_data_table, NewNumQuestsCompleted : int) <transacts>:player_data_table =
        NewData : player_data_table = player_data_table:
            MakePlayerDataTable<constructor>(ProfileData)
            NumQuestsCompleted := NewNumQuestsCompleted
        return NewData 

    #LORE
    SetLoreState(PlayerData : player_data_table, NumLore : int, NewValue : int) <transacts>:player_data_table =
        var AllLore : []int = PlayerData.AllLoreStates
        if(set AllLore[NumLore] = NewValue):
            NewData : player_data_table = player_data_table:
                MakePlayerDataTable<constructor>(PlayerData)
                AllLoreStates := AllLore
            return NewData
        else:
            return PlayerData

#INCREMENT PRODUCTS
    #Product Prepared_CheeseSalsaForPimentoCheeseSandwich
    <#
    Increment_Prepared_CheeseSalsaForPimentoCheeseSandwich (Player : player)<transacts>:int =
        var Output : int = 0
        if:
            PlayerData:=GetPlayerData[Player]
            NewData : player_data_table = player_data_table:
                MakePlayerDataTable<constructor>(PlayerData)
                Prepared_CheeseSalsaForPimentoCheeseSandwich := PlayerData.Prepared_CheeseSalsaForPimentoCheeseSandwich + 1
        then:
            set Output = NewData.Prepared_CheeseSalsaForPimentoCheeseSandwich
            SaveNewData(Player,PlayerData,NewData)
        return Output

    #Product Prepared_PimentosForPimentoCheeseSandwich
    Increment_Prepared_PimentosForPimentoCheeseSandwich (Player : player)<transacts>:int =
        var Output : int = 0
        if:
            PlayerData:=GetPlayerData[Player]
            NewData : player_data_table = player_data_table:
                MakePlayerDataTable<constructor>(PlayerData)
                Prepared_PimentosForPimentoCheeseSandwich := PlayerData.Prepared_PimentosForPimentoCheeseSandwich + 1
        then:
            set Output = NewData.Prepared_PimentosForPimentoCheeseSandwich
            SaveNewData(Player,PlayerData,NewData)
        return Output#>

#MASTERY
    SetMasteryQuestCompleted(PlayerData : player_data_table, NewValue : logic) <transacts>:player_data_table =
        NewData : player_data_table = player_data_table:
            MakePlayerDataTable<constructor>(PlayerData)
            IsMasteryQuestCompleted := NewValue
        return NewData

    SetNumMasteries(PlayerData : player_data_table, NumMasteries : int) <transacts>:player_data_table =
        NewData : player_data_table = player_data_table:
            MakePlayerDataTable<constructor>(PlayerData)
            NumMasteries := NumMasteries
        return NewData

    SetMasteryReset(PlayerData : player_data_table, AgedSapphiresGranted : float) <transacts>:player_data_table =
        NewData : player_data_table = player_data_table:
            MakePlayerDataTable<constructor>(PlayerData)
            #Map
            LV_Area := -1
            #AllSecrets_Granted := PlayerData.AllSecrets_Granted

            #History
            #Historic_BSent := 0.0
            #Historic_CBSent := 0.0
            #Historic_PCSSent := 0.0

            #MiniGames
            AllMiniGamesCooldowns := array{0,0,0,0,0,0,0,0,0,0,0,0,0,0,0,0,0,0,0,0,0,0,0,0,0,0,0,0,0,0,0,0,0,0,0,0,0,0,0,0,0,0,0,0,0,0,0,0,0,0,0,0,0,0,0,0,0,0,0,0,0,0,0,0,0,0,0,0,0,0,0,0,0,0,0,0,0,0,0,0,0,0,0,0,0,0,0,0,0,0,0,0,0,0,0,0,0,0,0,0,0,0,0,0,0,0,0,0,0,0,0,0,0,0,0,0,0,0,0,0}
            AllMiniGamesCompleted := array{false,false,false,false,false,false,false,false,false,false,false,false,false,false,false,false,false,false,false,false,false,false,false,false,false,false,false,false,false,false,false,false,false,false,false,false,false,false,false,false,false,false,false,false,false,false,false,false,false,false,false,false,false,false,false,false,false,false,false,false,false,false,false,false,false,false,false,false,false,false,false,false,false,false,false,false,false,false,false,false,false,false,false,false,false,false,false,false,false,false,false,false,false,false,false,false,false,false}

            #Inventory
            #Inventory_Fame := 0.0
            Inventory_Sapphires := 0.0
            Inventory_AgedSapphires := PlayerData.Inventory_AgedSapphires+AgedSapphiresGranted
            
            Inventory_Water := 0.0
            Inventory_Flour := 0.0
            Inventory_Cheese := 0.0
            Inventory_Pimentos := 0.0

            MainWaterDeposit_Value := 0.0
            MainFlourDeposit_Value := 0.0
            MainCheeseDeposit_Value := 0.0
            MainPimentosDeposit_Value := 0.0

            #Inventory_MutantMaterial := 0.0
            #Inventory_Coal := 0.0

            #OnBakers
            WaterPreparedFor_B := 0.0
            FlourPreparedFor_B := 0.0

            WaterPreparedFor_CB := 0.0
            FlourPreparedFor_CB := 0.0
            CheesePreparedFor_CB := 0.0

            WaterPreparedFor_PCS := 0.0
            FlourPreparedFor_PCS := 0.0
            CheesePreparedFor_PCS := 0.0
            PimentosPreparedFor_PCS := 0.0

            PB_Pimentos_Inside := 0.0
            PB_MutantMaterial_Inside  := 0.0

            #Weapons
                #Ammo
            Ammo_Arrows := 0
                #Sword
            #Tier_Sword := 1
            LV_SwordDamage := 1
            LV_SwordCriticalProb := 1
            LV_SwordCriticalDamage := 1
            LV_SwordReward := 1
                #Bow
            #Tier_Bow := 1
            LV_BowDamage := 1
            LV_BowCriticalProb := 1
            LV_BowCriticalDamage := 1
            LV_BowReward := 1

            #Resourcers backpack
            AllLV_BackpackLimits := array{0,0,0,0,0,0}

            #Pickaxe
            TLV_PickaxeStrength := array{0,-1,-1,-1,-1,-1,-1,-1,-1,-1,-1,-1,-1,-1,-1,-1,-1,-1,-1,-1}
            LV_PickaxeEfficiency := array{0,0,0,0,0}
            AllLV_PickaxePerks := array{0,0,0,0,1} #CriticalProb,CriticalDamage,PickaxeLuck,PickaxeWeakPointMult,PickaxeMultiResource

            #Drones
            B_On_Drone := 0.0
            CB_On_Drone := 0.0
            PCS_On_Drone := 0.0

            LV_DroneSpeed := 0
            LV_DroneStorage := 0

            #AllLV_SalesEfficiency : []int = array{0,0,0}

            TimeToDroneReturn := 0
            
                #Main desposits & AutoResourcers
            #SuperEnergy
            SuperEnergy_Owned := 0.0

            AllLV_MainDeposits := array{0,0,0,0}
            #AllLV_PasiveResourcers : []int = array{0,0,0,0}

                #Bread Bakery machine
            LV_BBSpeed := 0
            EfficiencyLV_BB := array{0,0}
            TLV_BB := array{0,-1,-1,-1,-1,-1,-1,-1,-1,-1,-1,-1,-1,-1,-1,-1,-1,-1,-1,-1}

                #Cheese Bread Bakery machine
            LV_CBBSpeed := 0
            EfficiencyLV_CBB := array{0,0,0}
            TLV_CBB := array{0,-1,-1,-1,-1,-1,-1,-1,-1,-1,-1,-1,-1,-1,-1,-1,-1,-1,-1,-1}

                #Pimento Cheese Sandwich Bakery machine
            LV_PCSBSpeed := 0
            EfficiencyLV_PCSB := array{0,0,0,0}
            TLV_PCSB := array{0,-1,-1,-1,-1,-1,-1,-1,-1,-1,-1,-1,-1,-1,-1,-1,-1,-1,-1,-1}

            #BAKER A
            #--------------
            #PRODUCTS
            Prod_Bread := 0
            Prepared_SlicedBreadForPimentoCheeseSandwich := 0
            #UPGRADES
            LV_BreadBaker := 0
            LV_SlicedBreadBaker := -1
            #BAKER B
            #--------------
            #PRODUCTS
            Prod_CheeseBread := 0
            Prepared_CheeseSalsaForPimentoCheeseSandwich := 0
            #RESOURCERS
            #UPGRADES
            LV_CheeseBreadBaker := -1
            LV_CheeseBlender := -1
            #BAKER C
            #--------------
            #PRODUCTS
            Prod_PimentoCheeseSandwich := 0
            #RESOURCERS
            Prepared_PimentosForPimentoCheeseSandwich := 0
            #UPGRADES
            LV_PimientoCheeseBlender := -1

            #PETS
            #AllPets : []pet_data_table=array{}
            #MaxActive_PetSlots := 1
            #MaxNoActive_PetSlots := 1
            #PetTier := 1
            
                #HexaGame
            #HexaGame_AllChallengesCompleted := array{}
            #HexaGame_ChallengeLVsUnlocked := 0
            HexaGame_Accumulated := 0
            #HexaGame_LVAccumulated := 1
            #HexaGame_LVDelay := 0
            
                
                #Dungeons
            #AllEvolvedDungeons : []int = array{0,0,0,0,0,0,0,0,0,0,0,0,0,0}
            #Dungeon_Completed : []int = array{}
            Dungeon_Accumulated := 0
            #Dungeon_LVAccumulated := 1
            #Dungeon_LVDelay := 0

                #Order assigned
            #OrderAssigned_ProgressAmount := -99.0
            #OrderAssigned_ProductAmountRequested := -99.0
            #OrderAssigned_ProductRequested := -99
            #OrderAssigned_Reward := -99.0
            #OrderAssigned_RewardType := -99

                #Quest assigned
            #QuestAssigned_NumQuest := -99
            #QuestAssigned_ObjetiveType := -99
            #QuestAssigned_ProgressAmount := 0.0
            #QuestAssigned_Reward := -99.0
            #QuestAssigned_RewardType := -99
            
                #Delays
            #HexaGame_DelayTime := 1.0
            #Dungeon_DelayTime := 1.0
            #Quest_DelayTime := 1.0
            Order_DelayTime := 1.0

                #Tutorial
            #NumQuestsCompleted := 0

###############OLD CODE#######################


    SetFame(ProfileData : player_data_table, NewFame:float)<transacts>:player_data_table =
        NewData : player_data_table = player_data_table:
            MakePlayerDataTable<constructor>(ProfileData)
            Inventory_Fame := NewFame

    #Basic Resourcers
    #Upgrades

  
    