player_data_table := class<final><persistable>:
    #Version
    Version:int = 1

    #Tutorial
    IsTutorialCompleted : logic = false

    #Language
    Language:int=0

    #Lore
    AllLoreStates : []int = array{0,0,0,0,0,0,0,0,0,0,0,0,0,0,0,0,0,0,0,0,0}

    #Options
    IsMusicOn : logic = true
    IsQuestTrackerOn : logic = true

    #Map
    LV_Area : int = -1
    LastAreaUnlocked : logic = false
    AllSecrets_Granted : []logic = array{false,false,false,false,false,false,false,false,false,false,false,false,false,false,false,false,false,false,false,false,false,false,false,false,false,false,false,false,false,false,false,false,false,false,false,false,false,false,false,false,false,false,false,false,false,false,false,false,false}

    #History
    Historic_BSent : float = 0.0
    Historic_CBSent : float = 0.0
    Historic_PCSSent : float = 0.0

    #MiniGames
    AllMiniGamesCooldowns : []int = array{0,0,0,0,0,0,0,0,0,0,0,0,0,0,0,0,0,0,0,0,0,0,0,0,0,0,0,0,0,0,0,0,0,0,0,0,0,0,0,0,0,0,0,0,0,0,0,0,0,0,0,0,0,0,0,0,0,0,0,0,0,0,0,0,0,0,0,0,0,0,0,0,0,0,0,0,0,0,0,0,0,0,0,0,0,0,0,0,0,0,0,0,0,0,0,0,0,0,0,0,0,0,0,0,0,0,0,0,0,0,0,0,0,0,0,0,0,0,0,0}
    AllMiniGamesCompleted :  []logic = array{false,false,false,false,false,false,false,false,false,false,false,false,false,false,false,false,false,false,false,false,false,false,false,false,false,false,false,false,false,false,false,false,false,false,false,false,false,false,false,false,false,false,false,false,false,false,false,false,false,false,false,false,false,false,false,false,false,false,false,false,false,false,false,false,false,false,false,false,false,false,false,false,false,false,false,false,false,false,false,false,false,false,false,false,false,false,false,false,false,false,false,false,false,false,false,false,false,false}

    #Inventory
    Inventory_Fame : float = 0.0
    Inventory_Sapphires : float = 0.0
    Inventory_AgedSapphires : float = 0.0
    
    Inventory_Water : float = 0.0
    Inventory_Flour : float = 0.0
    Inventory_Cheese : float = 0.0
    Inventory_Pimentos : float = 0.0

    MainWaterDeposit_Value : float = 0.0
    MainFlourDeposit_Value : float = 0.0
    MainCheeseDeposit_Value : float = 0.0
    MainPimentosDeposit_Value : float = 0.0

    Inventory_MutantMaterial : float = 0.0
    Inventory_Coal : float = 0.0

    #OnBakers
    WaterPreparedFor_B : float = 0.0
    FlourPreparedFor_B : float = 0.0

    WaterPreparedFor_CB : float = 0.0
    FlourPreparedFor_CB : float = 0.0
    CheesePreparedFor_CB : float = 0.0

    WaterPreparedFor_PCS : float = 0.0
    FlourPreparedFor_PCS : float = 0.0
    CheesePreparedFor_PCS : float = 0.0
    PimentosPreparedFor_PCS : float = 0.0

    PB_Pimentos_Inside : float = 0.0
    PB_MutantMaterial_Inside  : float = 0.0

    #Weapons
    #Ammo
    Ammo_Arrows : int = 0
        #Sword
    Tier_Sword : int = 1
    LV_SwordDamage : int = 1
    LV_SwordCriticalProb : int = 1
    LV_SwordCriticalDamage : int = 1
    LV_SwordReward : int = 1
        #Bow
    Tier_Bow : int = 1
    LV_BowDamage : int = 1
    LV_BowCriticalProb : int = 1
    LV_BowCriticalDamage : int = 1
    LV_BowReward : int = 1

    #Resourcers backpack
    AllLV_BackpackLimits : []int = array{0,0,0,0,0,0}

    #Pickaxe
    TLV_PickaxeStrength : []int = array{0,-1,-1,-1,-1,-1,-1,-1,-1,-1,-1,-1,-1,-1,-1,-1,-1,-1,-1,-1}
    LV_PickaxeEfficiency : []int = array{0,0,0,0,0}
    AllLV_PickaxePerks : []int = array{0,0,0,0,1} #CriticalProb,CriticalDamage,PickaxeLuck,PickaxeWeakPointMult,PickaxeMultiResource

    #Drones
    B_On_Drone : float = 0.0
    CB_On_Drone : float = 0.0
    PCS_On_Drone : float = 0.0

    LV_DroneSpeed : int = 0
    LV_DroneStorage : int = 0

    AllLV_SalesEfficiency : []int = array{0,0,0}

    TimeToDroneReturn : int = 0
    
        #Main desposits & AutoResourcers
    #SuperEnergy
    SuperEnergy_Owned : float = 0.0

    AllLV_MainDeposits : []int = array{0,0,0,0}
    AllLV_PasiveResourcers : []int = array{0,0,0,0}

        #Bread Bakery machine
    LV_BBSpeed : int = 0
    EfficiencyLV_BB :[]int = array{0,0}
    TLV_BB : []int = array{0,-1,-1,-1,-1,-1,-1,-1,-1,-1,-1,-1,-1,-1,-1,-1,-1,-1,-1,-1}

        #Cheese Bread Bakery machine
    LV_CBBSpeed : int = 0
    EfficiencyLV_CBB :[]int = array{0,0,0}
    TLV_CBB : []int = array{0,-1,-1,-1,-1,-1,-1,-1,-1,-1,-1,-1,-1,-1,-1,-1,-1,-1,-1,-1}

        #Pimento Cheese Sandwich Bakery machine
    LV_PCSBSpeed : int = 0
    EfficiencyLV_PCSB :[]int = array{0,0,0,0}
    TLV_PCSB : []int = array{0,-1,-1,-1,-1,-1,-1,-1,-1,-1,-1,-1,-1,-1,-1,-1,-1,-1,-1,-1}

    #BAKER A
    #--------------
    #PRODUCTS
    Prod_Bread : int = 0
    Prepared_SlicedBreadForPimentoCheeseSandwich : int = 0
    #UPGRADES
    LV_BreadBaker : int = 0
    LV_SlicedBreadBaker : int = -1
    #BAKER B
    #--------------
    #PRODUCTS
    Prod_CheeseBread : int = 0
    Prepared_CheeseSalsaForPimentoCheeseSandwich : int = 0
    #RESOURCERS
    #UPGRADES
    LV_CheeseBreadBaker : int = -1
    LV_CheeseBlender : int = -1
    #BAKER C
    #--------------
    #PRODUCTS
    Prod_PimentoCheeseSandwich : int = 0
    #RESOURCERS
    Prepared_PimentosForPimentoCheeseSandwich : int = 0
    #UPGRADES
    LV_PimientoCheeseBlender : int = -1

    #PETS
    AllPets : []pet_data_table=array{}
    MaxActive_PetSlots : int = 1
    MaxNoActive_PetSlots : int = 1
    PetTier : int = 1
    
        #HexaGame
    HexaGame_AllChallengesCompleted : []int = array{}
    AllHexaGame_TotemsUnlocked : []logic = array{false,false,false,false,false,false,false,false,false,false,false,false,false,false,false,false,false,false,false,false,false}
    HexaGame_Accumulated : int = 0
    HexaGame_LVAccumulated : int = 1
    HexaGame_LVDelay : int = 0

        #HexaGame Pointes
    HexaGame_ChallengePointer : int = 0
        #Dungeons
    AllEvolvedDungeons : []int = array{0,0,0,0,0,0,0,0,0,0,0,0,0,0}
    Dungeon_Completed : []int = array{}
    Dungeon_Accumulated : int = 0
    Dungeon_LVAccumulated : int = 1
    Dungeon_LVDelay : int = 0

        #Order assigned
    OrderAssigned_ProgressAmount : float = -99.0
    OrderAssigned_ProductAmountRequested : float = -99.0
    OrderAssigned_ProductRequested : int = -99
    OrderAssigned_Reward : float = -99.0
    OrderAssigned_RewardType : int = -99

        #Quest assigned
    QuestAssigned_NumQuest : int = -99
    QuestAssigned_ObjetiveType : int = -99
    QuestAssigned_ProgressAmount : float = 0.0
    QuestAssigned_Reward : float = -99.0
    QuestAssigned_RewardType : int = -99

    NumQuestsCompleted : int = 0

    NumMasteries : int = 0
    IsMasteryQuestCompleted : logic = false
    
        #Delays
    HexaGame_DelayTime : float = 1.0
    Dungeon_DelayTime : float = 1.0
    Quest_DelayTime : float = 1.0
    Order_DelayTime : float = 1.0

        #Tutorial
    

MakePlayerDataTable<constructor>(OldTable:player_data_table)<transacts> := player_data_table:
    Version := OldTable.Version
    IsTutorialCompleted := OldTable.IsTutorialCompleted
    Language := OldTable.Language
    #Lore
    AllLoreStates := OldTable.AllLoreStates
    #Options
    IsMusicOn := OldTable.IsMusicOn
    IsQuestTrackerOn := OldTable.IsQuestTrackerOn
    #MAP
    LV_Area := OldTable.LV_Area
    LastAreaUnlocked := OldTable.LastAreaUnlocked
    AllSecrets_Granted := OldTable.AllSecrets_Granted
    #HISTORIC
    Historic_BSent :=OldTable.Historic_BSent
    Historic_CBSent :=OldTable.Historic_CBSent
    Historic_PCSSent :=OldTable.Historic_PCSSent
    #MINIGAMES
    AllMiniGamesCooldowns := OldTable.AllMiniGamesCooldowns
    #DRONE
    B_On_Drone := OldTable.B_On_Drone
    CB_On_Drone := OldTable.CB_On_Drone
    PCS_On_Drone := OldTable.PCS_On_Drone

    LV_DroneSpeed := OldTable.LV_DroneSpeed
    LV_DroneStorage := OldTable.LV_DroneStorage

    AllLV_SalesEfficiency := OldTable.AllLV_SalesEfficiency

    #INVENTORY
    Inventory_Fame := OldTable.Inventory_Fame
    Inventory_Sapphires := OldTable.Inventory_Sapphires
    Inventory_AgedSapphires := OldTable.Inventory_AgedSapphires
    
    Inventory_Water := OldTable.Inventory_Water
    Inventory_Flour := OldTable.Inventory_Flour
    Inventory_Cheese := OldTable.Inventory_Cheese
    Inventory_Pimentos := OldTable.Inventory_Pimentos

    MainWaterDeposit_Value := OldTable.MainWaterDeposit_Value
    MainFlourDeposit_Value := OldTable.MainFlourDeposit_Value
    MainCheeseDeposit_Value := OldTable.MainCheeseDeposit_Value
    MainPimentosDeposit_Value := OldTable.MainPimentosDeposit_Value

    Inventory_MutantMaterial := OldTable.Inventory_MutantMaterial
    Inventory_Coal := OldTable.Inventory_Coal
    AllLV_BackpackLimits := OldTable.AllLV_BackpackLimits

#INSIDE BAKERS
    WaterPreparedFor_B := OldTable.WaterPreparedFor_B
    FlourPreparedFor_B := OldTable.FlourPreparedFor_B

    WaterPreparedFor_CB := OldTable.WaterPreparedFor_CB
    FlourPreparedFor_CB := OldTable.FlourPreparedFor_CB
    CheesePreparedFor_CB := OldTable.CheesePreparedFor_CB

    WaterPreparedFor_PCS := OldTable.WaterPreparedFor_PCS
    FlourPreparedFor_PCS := OldTable.FlourPreparedFor_PCS
    CheesePreparedFor_PCS := OldTable.CheesePreparedFor_PCS
    PimentosPreparedFor_PCS := OldTable.PimentosPreparedFor_PCS

    PB_Pimentos_Inside := OldTable.PB_Pimentos_Inside
    PB_MutantMaterial_Inside  := OldTable.PB_MutantMaterial_Inside

#BASIC RESOURCERS
    #BAKER A
    #--------------
    #PRODUCTS
    Prod_Bread := OldTable.Prod_Bread
    Prepared_SlicedBreadForPimentoCheeseSandwich := OldTable.Prepared_SlicedBreadForPimentoCheeseSandwich
    #UPGRADES
    LV_BreadBaker := OldTable.LV_BreadBaker
    LV_SlicedBreadBaker := OldTable.LV_SlicedBreadBaker
    #BAKER B
    #--------------
    #PRODUCTS
    Prod_CheeseBread := OldTable.Prod_CheeseBread
    Prepared_CheeseSalsaForPimentoCheeseSandwich := OldTable.Prepared_CheeseSalsaForPimentoCheeseSandwich
    #RESOURCERS
    #UPGRADES
    LV_CheeseBreadBaker := OldTable.LV_CheeseBreadBaker
    LV_CheeseBlender := OldTable.LV_CheeseBlender
#BAKERS
    #--------------
    #SUPERENERGY
    SuperEnergy_Owned := OldTable.SuperEnergy_Owned
    #PRODUCTS
    Prod_PimentoCheeseSandwich := OldTable.Prod_PimentoCheeseSandwich
    #RESOURCERS
    Prepared_PimentosForPimentoCheeseSandwich := OldTable.Prepared_PimentosForPimentoCheeseSandwich
    #Speed
    LV_PCSBSpeed:=OldTable.LV_PCSBSpeed
    LV_CBBSpeed:=OldTable.LV_CBBSpeed
    LV_BBSpeed:=OldTable.LV_BBSpeed
    #Upgrade production
    TLV_PCSB := OldTable.TLV_PCSB
    TLV_CBB := OldTable.TLV_CBB
    TLV_BB := OldTable.TLV_BB
    #Efficiency
    EfficiencyLV_PCSB:=OldTable.EfficiencyLV_PCSB
    EfficiencyLV_CBB:=OldTable.EfficiencyLV_CBB
    EfficiencyLV_BB:=OldTable.EfficiencyLV_BB
#PICKAXE
    TLV_PickaxeStrength:=OldTable.TLV_PickaxeStrength
    LV_PickaxeEfficiency:=OldTable.LV_PickaxeEfficiency
    AllLV_PickaxePerks:=OldTable.AllLV_PickaxePerks
#MAIN DEPOSITS AND PASIVE INCOME
    AllLV_MainDeposits := OldTable.AllLV_MainDeposits
    AllLV_PasiveResourcers := OldTable.AllLV_PasiveResourcers
#PETS
    AllPets := OldTable.AllPets
    MaxNoActive_PetSlots := OldTable.MaxNoActive_PetSlots
    MaxActive_PetSlots := OldTable.MaxActive_PetSlots
    PetTier := OldTable.PetTier
#HEXAGAME
    HexaGame_AllChallengesCompleted := OldTable.HexaGame_AllChallengesCompleted
    AllHexaGame_TotemsUnlocked := OldTable.AllHexaGame_TotemsUnlocked
    HexaGame_Accumulated := OldTable.HexaGame_Accumulated
    HexaGame_LVAccumulated := OldTable.HexaGame_LVAccumulated
    HexaGame_LVDelay := OldTable.HexaGame_LVDelay
    HexaGame_DelayTime := OldTable.HexaGame_DelayTime
    #HexaGame Pointer
    HexaGame_ChallengePointer := OldTable.HexaGame_ChallengePointer
#DUNGEONS
    AllEvolvedDungeons := OldTable.AllEvolvedDungeons
    Dungeon_Completed := OldTable.Dungeon_Completed
    Dungeon_Accumulated := OldTable.Dungeon_Accumulated
    Dungeon_LVAccumulated := OldTable.Dungeon_LVAccumulated
    Dungeon_LVDelay := OldTable.Dungeon_LVDelay
    Dungeon_DelayTime := OldTable.Dungeon_DelayTime

#WEAPONS
    Ammo_Arrows:=OldTable.Ammo_Arrows
    #Sword
    Tier_Sword:=OldTable.Tier_Sword
    LV_SwordDamage := OldTable.LV_SwordDamage
    LV_SwordCriticalProb := OldTable.LV_SwordCriticalProb
    LV_SwordCriticalDamage := OldTable.LV_SwordCriticalDamage
    LV_SwordReward := OldTable.LV_SwordReward
    #Bow
    Tier_Bow := OldTable.Tier_Bow
    LV_BowDamage := OldTable.LV_BowDamage
    LV_BowCriticalProb := OldTable.LV_BowCriticalProb
    LV_BowCriticalDamage := OldTable.LV_BowCriticalDamage
    LV_BowReward := OldTable.LV_BowReward

    #Quests and orders
    Quest_DelayTime := OldTable.Quest_DelayTime
    Order_DelayTime := OldTable.Order_DelayTime

        #Order assigned
    OrderAssigned_ProgressAmount := OldTable.OrderAssigned_ProgressAmount
    OrderAssigned_ProductAmountRequested := OldTable.OrderAssigned_ProductAmountRequested
    OrderAssigned_ProductRequested := OldTable.OrderAssigned_ProductRequested
    OrderAssigned_Reward := OldTable.OrderAssigned_Reward
    OrderAssigned_RewardType := OldTable.OrderAssigned_RewardType
    
        #Quest assigned
    QuestAssigned_NumQuest := OldTable.QuestAssigned_NumQuest
    QuestAssigned_ProgressAmount := OldTable.QuestAssigned_ProgressAmount
    QuestAssigned_ObjetiveType := OldTable.QuestAssigned_ObjetiveType
    QuestAssigned_Reward := OldTable.QuestAssigned_Reward
    QuestAssigned_RewardType := OldTable.QuestAssigned_RewardType

        #Masteries
    NumMasteries  := OldTable.NumMasteries
    IsMasteryQuestCompleted := OldTable.IsMasteryQuestCompleted

        #Tutorials / Quests
    NumQuestsCompleted := OldTable.NumQuestsCompleted
