#Deposit limits
MainFormulaBackpackCost_Limit(CostBase : float, Level : float) : float = 
    if(Level<=0.0):
        return CostBase/2.0
    else:
        return Pow(10.0,Level)

MainFormulaBackpack_Limit(CostLimit : float) : float = 
    return CostLimit * 100.0


#Sales efficiency
MainFormulaSales_Efficiency(Level : float) : float =
    if(Level <= 0.0):
        return 0.0
    else:
        SalesBaseEfficiency * Level

MainFormulaSalesCost_Efficiency(CurrentLevel : float) : float =
    #if(CurrentLevel<1.0):
    #    return 1.0
    #else:
        return Sales_Efficiency_IncrementCost*CurrentLevel

#Drones
MainFormulaDroneCost_Speed(Level : int) : float =
    return Pow(DroneBaker_Speed_IncrementCost,Level*1.0)

MainFormulaDrone_Speed(Level : int) : int =
    return Level * Base_Drone_UpgradeSpeed

#PasiveIncome 
MainFormulaMainPasiveIncomeCost(CostBase : float, Level : float) : float = 
    var Cost : float = 0.0
    if(Level<=0.0):
        return CostBase - 3.0
    return CostBase * Level

MainFormulaMainPasiveIncome(CostBase : float, Level : float) : float = 
    return CostBase * Level

#Deposit limits
MainFormulaMainDepositCost_Limit(CostBase : float, Level : float) : float = 
    if(Level<=0.0):
        return CostBase/2.0
    else:
        return Pow(10.0,Level)

MainFormulaMainDeposit_Limit(CostLimit : float) : float = 
    return CostLimit * 100.0

#Bakers
MainFormulaBakerCost_Production(CostBase : float, Tier : float, Level : float) : float = 
    var Ammount : float = Pow(Level,2.0)*Pow(CostBase,Tier)
    if(Tier >= 0.0):
        set Ammount = Ammount + 10.0*Pow(CostBase,Tier-1.0)
    if(Ammount<1000.0):
        if(IntAmmount := Ceil[Ammount], set Ammount = IntAmmount * 1.0):
    return Ammount

MainFormulaBaker_Production(Cost : float, Level : float, Tier : float) : float =
    if(Level=-1.0 or Level=0.0):
        return 0.0
    else:
        var Ammount : float = 0.0
        if(Cost<1000.0):
            set Ammount = Cost
        else:
            #set Ammount = Cost * (Pow(BakerLevelDegradation,Level))
            set Ammount = Cost * (Pow(PickaxeLevelDegradation,Level)) * (Pow(PickaxeTierDegradation,Tier))
        if(Ammount<1000.0, IntAmount:=Floor[Ammount]):
            set Ammount=IntAmount*1.0

        return Ammount

MainFormulaBakerCost_Speed(Level : float) : float =
    return Pow(Baker_Speed_IncrementCost,Level)

MainFormulaBaker_Speed(Level : float) : float =
    return Level * Baker_Speed_Increment

MainFormulaBaker_Efficiency(Level : float) : float =
    if(Level <= 0.0):
        return BakerBaseEfficiency / 2.0
    else:
        BakerBaseEfficiency * Level

MainFormulaBakerCost_Efficiency(ResourceType : int, CurrentLevel : float) : float =
    Ammount:=Pow(Baker_Efficiency_IncrementCost,CurrentLevel)
    return Ammount

MainFormulaAutoResource(Level : int) : float =
    if(Level=0):
        return 0.05
    return Level * 0.15

CalculeSpeedLimitLevel(BaseSpeed : float) : int =
    var MaxLevel : int = 0
    if(Value := Floor[BaseSpeed / MinBakeTime]):
        set MaxLevel = Value - 1
    return MaxLevel

CalculeBakerSpeed(BakerMachine : int, CurrentLevel : int, BuyModeMultiplier : int) : tuple(float,float,float) =
    var BaseSpeed : float = 0.0
    var MaxSpeedLV : int = 0
    if(BakerMachine=BB_TYPE):
        set BaseSpeed = Base_BB_Speed
        set MaxSpeedLV = CalculeSpeedLimitLevel(Base_BB_Speed)
    else if(BakerMachine=CBB_TYPE):
        set MaxSpeedLV = CalculeSpeedLimitLevel(Base_CBB_Speed)
        set BaseSpeed = Base_CBB_Speed
    else if(BakerMachine=PCSB_TYPE): 
        set MaxSpeedLV = CalculeSpeedLimitLevel(Base_PCSB_Speed)
        set BaseSpeed = Base_PCSB_Speed
    else:
        PrintFatalError("FATAL ERROR on CalculeBakerSpeed. BakerMachine {BakerMachine} not valid!")

    #### Next stats
    var ObjetiveLevel : int = CurrentLevel+BuyModeMultiplier
    if(ObjetiveLevel > MaxSpeedLV):
        set ObjetiveLevel = MaxSpeedLV
    var TotalNextCost : float = 0.0
    for(Level:=CurrentLevel..ObjetiveLevel-1):
        set TotalNextCost = TotalNextCost + MainFormulaBakerCost_Speed(Level*1.0)

    ObjetiveCost := MainFormulaBakerCost_Speed(ObjetiveLevel*1.0)
    NextSpeed := MainFormulaBaker_Speed(ObjetiveLevel*1.0)
    #### Current stats
    CurrentCost:=MainFormulaBakerCost_Speed(CurrentLevel*1.0)
    CurrentSpeed := MainFormulaBaker_Speed(CurrentLevel*1.0)

    #### Difference
    TotalDifference := NextSpeed-CurrentSpeed
    return(BaseSpeed-CurrentSpeed,TotalNextCost,TotalDifference)

#Efficiency
CalculeBakerResourceEfficiency<public>(CurrentLevel : int, BakerMachine : int, ResourceType : int, BuyModeMultiplier : int) : tuple(float,float,float) =
    #### Next stats
    var ObjetiveLevel : int = CurrentLevel+BuyModeMultiplier
    if(ObjetiveLevel > MaxBakerEfficiencyLV):
        set ObjetiveLevel = MaxBakerEfficiencyLV
    var TotalNextCost : float = 0.0
    for(Level:=CurrentLevel..ObjetiveLevel-1):
        set TotalNextCost = TotalNextCost + MainFormulaBakerCost_Efficiency(ResourceType, Level*1.0)

    ObjetiveCost := MainFormulaBakerCost_Efficiency(ResourceType, ObjetiveLevel*1.0)
    NextProduction := MainFormulaBaker_Efficiency(ObjetiveLevel*1.0)
    #### Current stats
    CurrentCost:=MainFormulaBakerCost_Efficiency(ResourceType, CurrentLevel*1.0)
    CurrentProduction := MainFormulaBaker_Efficiency(CurrentLevel*1.0)

    #### Difference
    TotalDifference := NextProduction-CurrentProduction
    return(CurrentProduction,TotalNextCost,TotalDifference)

CalculeTotalProduction<public>(TLV : []int, BakerMachine : int) : float =
    var Base : float = 0.0
    var Total : float = 0.0
    case(BakerMachine):
        0=> set Base = Base_B_Production * 1.0 
        1=> set Base = Base_CB_Production * 1.0
        2=> set Base = Base_PCS_Production * 1.0
        _=> #Print("Fatal error on CalculeProductionCostUpgrade")
    for(Tier := 0..TLV.Length-1, Level := TLV[Tier], Level <> -1):
        var LevelToCalculate : int = Level
        if(Level>MaxTierLevel):
            set LevelToCalculate = MaxTierLevel

        Cost:=MainFormulaBakerCost_Production(Base, Tier*1.0, LevelToCalculate*1.0)
        TierProduction := MainFormulaBaker_Production(Cost, LevelToCalculate*1.0, Tier*1.0)
        set Total = Total + TierProduction
    if(Total=0.0):
        return 1.0
    return Total

CalculeBakerUpgradeProduction<public>(Tier : int, CurrentLevel : int, BakerMachine : int, BuyModeMultiplier : int) : tuple(float,float,float,float) =
    var Base : float = 0.0
    case(BakerMachine):
        0=> set Base = Base_B_Production * 1.0
        1=> set Base = Base_CB_Production * 1.0
        2=> set Base = Base_PCS_Production * 1.0
        _=> #Print("Fatal error on CalculeProductionCostUpgrade")

    #### Next stats
    var ObjetiveLevel : int = CurrentLevel+BuyModeMultiplier
    if(ObjetiveLevel > MaxTierLevel):
        set ObjetiveLevel = MaxTierLevel
    var TotalNextCost : float = 0.0
    for(Level:=CurrentLevel..ObjetiveLevel-1):
        set TotalNextCost = TotalNextCost + MainFormulaBakerCost_Production(Base, Tier*1.0, Level*1.0)

    ObjetiveCost := MainFormulaBakerCost_Production(Base, Tier*1.0, ObjetiveLevel*1.0)
    NextProduction := MainFormulaBaker_Production(ObjetiveCost, ObjetiveLevel*1.0, Tier*1.0)
    #### Current stats
    CurrentCost:=MainFormulaBakerCost_Production(Base, Tier*1.0, CurrentLevel*1.0)
    CurrentProduction := MainFormulaBaker_Production(CurrentCost,CurrentLevel*1.0, Tier*1.0)

    #### Difference
    TotalDifference := NextProduction-CurrentProduction
    return(CurrentProduction,TotalNextCost,TotalDifference,NextProduction)


CalculeDepositUpgradeLimit<public>(CurrentLevel : int, ResourceDeposit : int) : tuple(float,float,float) =
    var Base : float = 0.0
    case(ResourceDeposit):
        0=> set Base = Base_MainWaterDepositCostLimit
        1=> set Base = Base_MainFlourDepositCostLimit
        2=> set Base = Base_MainCheeseDepositCostLimit
        3=> set Base = Base_MainPimentosDepositCostLimit
        _=> #Print("Fatal error on CalculeProductionCostUpgrade")

    #### Next stats
    var ObjetiveLevel : int = CurrentLevel+1
    CurrentCost :=MainFormulaMainDepositCost_Limit(Base, CurrentLevel*1.0)
    ObjetiveCost :=MainFormulaMainDepositCost_Limit(Base, ObjetiveLevel*1.0)

    CurrentLimit := MainFormulaMainDeposit_Limit(CurrentCost)
    NextLimit  := MainFormulaMainDeposit_Limit(ObjetiveCost)

    #### Difference
    TotalDifference := NextLimit-CurrentLimit
    return(CurrentLimit,ObjetiveCost,TotalDifference)

CalculeDepositPasiveIncome(CurrentLevel : int) : tuple(float,float,float) =
    #### Next stats
    var ObjetiveLevel : int = CurrentLevel+1
    if(ObjetiveLevel > MaxUpgradePasiveIncome):
        set ObjetiveLevel = MaxUpgradePasiveIncome
    ObjetiveCost :=MainFormulaMainPasiveIncomeCost(Base_PasiveIncomeCost, ObjetiveLevel*1.0)

    CurrentPasive := MainFormulaMainPasiveIncome(Base_PasiveIncome, CurrentLevel*1.0)/100.0
    NextPasive  := MainFormulaMainPasiveIncome(Base_PasiveIncome, ObjetiveLevel*1.0)/100.0

    #### Difference
    TotalDifference := NextPasive-CurrentPasive
    return(CurrentPasive,ObjetiveCost,TotalDifference)

#Drone stats
CalculeDroneSpeedLimitLevel() : int =
    var MaxLevel : int = 0
    if(Value := Floor[(Base_Drone_Speed*1.0 - Base_Drone_MaxSpeed*1.0) / (Base_Drone_UpgradeSpeed*1.0)]):
        return Value
    return MaxLevel

CalculeDroneSpeed(CurrentLevel : int) : tuple(int,float,int) =
    MaxSpeedLV : int = CalculeDroneSpeedLimitLevel() 
    #### Next stats
    var ObjetiveLevel : int = CurrentLevel+1
    if(ObjetiveLevel > MaxSpeedLV):
        set ObjetiveLevel = MaxSpeedLV

    ObjetiveCost := MainFormulaDroneCost_Speed(ObjetiveLevel)
    NextSpeed := MainFormulaDrone_Speed(ObjetiveLevel)
    #### Current stats
    CurrentCost:=MainFormulaDroneCost_Speed(CurrentLevel)
    CurrentSpeed := MainFormulaDrone_Speed(CurrentLevel)

    #### Difference
    TotalDifference := NextSpeed-CurrentSpeed
    return(Base_Drone_Speed-CurrentSpeed,ObjetiveCost,TotalDifference)

#Drone Storage
CalculeDroneStorage(CurrentProduction : float, Speed : float) : tuple(float,float,float) =
    return ((CurrentProduction*(1.0/Speed))*500.0,0.0,0.0)

    #OldCode
    <#
    var MaxSpeedLV : int = 0
    var BaseType : float = 0.0
    case(Type):
        0=> set BaseType = Base_B_Storage
        1=> set BaseType = Base_CB_Storage
        2=> set BaseType = Base_PCS_Storage
        _=> #Print("Fatal error on CalculeDroneStorage, Type: {Type} not valid")

    #### Next stats
    var ObjetiveLevel : int = CurrentLevel+1
    if(ObjetiveLevel > MaxSpeedLV):
        set ObjetiveLevel = MaxSpeedLV

    ObjetiveCost := MainFormulaMainDepositCost_Limit(BaseType,ObjetiveLevel*1.0)
    NextStorage := MainFormulaMainDeposit_Limit(ObjetiveCost)
    #### Current stats
    CurrentCost:=MainFormulaMainDepositCost_Limit(BaseType,CurrentLevel*1.0)
    CurrentStorage := MainFormulaMainDeposit_Limit(CurrentCost)

    #### Difference
    TotalDifference := NextStorage-CurrentStorage
    return(TotalDifference,ObjetiveCost,TotalDifference)#>

RecalculeNewDroneSpeed(OldSpeed : int, NewSpeed : int, TimeToDroneReturn : int) : int =
    TimeRemaining := TimeToDroneReturn-(OldSpeed-NewSpeed)
    if(TimeRemaining<0):
        return 0
    else:
        return TimeRemaining

#Sales efficiency
CalculeSalesEfficiency<public>(CurrentLevel : int) : tuple(float,float,float) =
    #### Next stats
    var ObjetiveLevel : int = CurrentLevel+1
    if(ObjetiveLevel > MaxSalesLV):
        set ObjetiveLevel = MaxSalesLV
    
    ObjetiveCost := MainFormulaSalesCost_Efficiency((ObjetiveLevel+1)*1.0)
    NextSales := MainFormulaSales_Efficiency(ObjetiveLevel*1.0)
    #### Current stats
    CurrentCost:=MainFormulaSalesCost_Efficiency(CurrentLevel*1.0)
    CurrentSales := MainFormulaSales_Efficiency(CurrentLevel*1.0)

    #### Difference
    TotalDifference := NextSales-CurrentSales
    return(CurrentSales,ObjetiveCost,TotalDifference)








