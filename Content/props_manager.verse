
using { /Fortnite.com/Devices }
using { /Verse.org/Simulation }
using { /UnrealEngine.com/Temporary/Diagnostics }
using { /UnrealEngine.com/Temporary/SpatialMath }

# See https://dev.epicgames.com/documentation/en-us/uefn/create-your-own-device-in-verse for how to create a verse device.

# A Verse-authored creative device that can be placed in a level
props_manager := class(creative_device):

    #Teleportation
    @editable
    AllTeleportationButtons : []button_device = array{}
    @editable
    AllTeleportationProps : []creative_prop = array{}

    #Pets
    @editable
    AllPetPropAsset : []creative_prop_asset = array{}

    #Super Energy Devices
    @editable
    AllTimerDevices : []timer_device = array{}
    @editable
    AllOverPoweredSigns : []creative_prop =array{}

    #PopDialogDevice
    #@editable
    #AllPopDialogDevices_ComputerInfo : []popup_dialog_device = array{}

    #BakersProps
    @editable
    AllBBProps : []creative_prop = array{}
    @editable
    AllCBBProps : []creative_prop = array{}
    @editable
    AllPCSBProps : []creative_prop = array{}
    @editable
    AllDroneProps : []creative_prop = array{}

    <#
#Issues
    @editable
    AllReparationChairs : []chair_device = array{}
    @editable
    InputTrigger_A : input_trigger_device = input_trigger_device{}
    @editable
    InputTrigger_B : input_trigger_device = input_trigger_device{}

    @editable
    AllPropsManipulator_WaterIssue : []prop_manipulator_device = array{}

    @editable
    AllButtons_FlourIssue : []button_device = array{}
    @editable
    AllCinematic_FlourIssue : []cinematic_sequence_device = array{}

    @editable
    AllButtons_CheeseIssue : []button_device = array{}
    @editable
    AllCinematic_CheeseIssue : []cinematic_sequence_device = array{}
   
    @editable
    AllButtons_PimentoIssue : []button_device = array{}
    @editable
    AllCinematic_PimentoIssue : []cinematic_sequence_device = array{}#>

    @editable
    AllPlayerReferences : []player_reference_device = array{}

#AllBillboard
    @editable
    AllBillboards_WaterDeposit : []billboard_device = array{}
    @editable
    AllBillboards_FlourDeposit : []billboard_device = array{}
    @editable
    AllBillboards_CheeseDeposit : []billboard_device = array{}
    @editable
    AllBillboards_PimentosDeposit : []billboard_device = array{}

    @editable
    AllBillboards_DroneDeposit : []billboard_device = array{}
    @editable
    AllBillboards_DroneShip : []billboard_device = array{}

#Deposits
    @editable
    AllMainWaterDeposit_AmmountProp : []creative_prop = array{}
    @editable
    AllMainFlourDeposit_AmmountProp : []creative_prop = array{}
    @editable
    AllMainCheeseDeposit_AmmountProp : []creative_prop = array{}
    @editable
    AllMainPimentosDeposit_AmmountProp : []creative_prop = array{}
    @editable
    AllButtons_Funnel : []button_device = array{}

#Computers
    @editable
    AllButtons_BB_Upgrades : []button_device = array{}
    @editable
    AllButtons_CBB_Upgrades : []button_device = array{}
    @editable
    AllButtons_PCSB_Upgrades : []button_device = array{}

#BakerInfo buttons
    @editable
    AllButtons_BB_BakerInfo : []button_device = array{}
    @editable
    AllButtons_CBB_BakerInfo : []button_device = array{}
    @editable
    AllButtons_PCSB_BakerInfo : []button_device = array{}
#Deposits
    @editable
    AllButtons_UpgradeDeposit : []button_device = array{}
#Drones
    @editable
    AllButtons_UpgradeDrone : []button_device = array{}
#Drone ship
    @editable
    AllButtons_ShipDrone : []button_device = array{}
    <#
#ProgressBars
    @editable
    AllProps_ProgressBar_B : []creative_prop = array{}
    @editable
    AllProps_ProgressBar_CB : []creative_prop = array{}
    @editable
    AllProps_ProgressBar_PCSB : []creative_prop = array{}#>


#Valves 
    #Buttons
    @editable
    AllValveButtons_WaterFor_B : []button_device = array{}
    @editable
    AllValveButtons_FlourFor_B : []button_device = array{}
    @editable
    AllValveButtons_WaterFor_CB : []button_device = array{}
    @editable
    AllValveButtons_FlourFor_CB : []button_device = array{}
    @editable
    AllValveButtons_CheeseFor_CB : []button_device = array{}
    @editable
    AllValveButtons_WaterFor_PCSB : []button_device = array{}
    @editable
    AllValveButtons_FlourFor_PCSB : []button_device = array{}
    @editable
    AllValveButtons_CheeseFor_PCSB : []button_device = array{}
    #@editable
    #AllValveButtons_PimentosFor_PCSB : []button_device = array{}

    #Simplified VALVES
    @editable
    AllSimplified_ValveButtons_B : []button_device = array{}
    @editable
    AllSimplified_ValveButtons_CB : []button_device = array{}
    @editable
    AllSimplified_ValveButtons_PCS : []button_device = array{}

    #Props
    @editable
    AllValveProps_WaterFor_B : []creative_prop = array{}
    @editable
    AllValveProps_FlourFor_B : []creative_prop = array{}
    @editable
    AllValveProps_WaterFor_CB : []creative_prop = array{}
    @editable
    AllValveProps_FlourFor_CB : []creative_prop = array{}
    @editable
    AllValveProps_CheeseFor_CB : []creative_prop = array{}
    @editable
    AllValveProps_WaterFor_PCS : []creative_prop = array{}
    @editable
    AllValveProps_FlourFor_PCS : []creative_prop = array{}
    @editable
    AllValveProps_CheeseFor_PCS : []creative_prop = array{}
    @editable
    AllValveProps_PimentosFor_PCS : []creative_prop = array{}

    GetPlayerReferenceDevice(Index : int) : player_reference_device =
        var PlayerReference : player_reference_device = player_reference_device{}
        if(Reference := AllPlayerReferences[Index]):
            set PlayerReference = Reference
        return PlayerReference
            
    GetPetProp(PetType : int) : creative_prop_asset =
        if(PropAsset := AllPetPropAsset[PetType]):
            return PropAsset
        return DefaultCreativePropAsset

    GetTeleportDevices(Index : int) : tuple(button_device,creative_prop) =
        var OutPutButton : button_device = button_device{}
        var OutPutProp : creative_prop = creative_prop{}
        if:
            Button:=AllTeleportationButtons[Index]
            Prop:=AllTeleportationProps[Index]
        then:
            set OutPutButton = Button
            set OutPutProp = Prop
        return (OutPutButton,OutPutProp)

    GetValves(Index : int) : tuple ([]creative_prop,[]button_device,[]button_device) =
        var ButtonsArray : []button_device = array{}
        var SimplifiedButtonsArray : []button_device = array{}
        var PropsArray : []creative_prop = array{}
        if:
            #Simplified
            Simplified_ValveButtons_B := AllSimplified_ValveButtons_B[Index]
            Simplified_ValveButtons_CB := AllSimplified_ValveButtons_CB[Index]
            Simplified_ValveButtons_PCS := AllSimplified_ValveButtons_PCS[Index]
            #BakerA
            ValveButton_WaterFor_B := AllValveButtons_WaterFor_B[Index]
            ValveButton_FlourFor_B := AllValveButtons_FlourFor_B[Index]
            #BakerB
            ValveButton_WaterFor_CB := AllValveButtons_WaterFor_CB[Index]
            ValveButton_FlourFor_CB := AllValveButtons_FlourFor_CB[Index]
            ValveButton_CheeseFor_CB := AllValveButtons_CheeseFor_CB[Index]
            #BakerC
            ValveButton_WaterFor_PCS := AllValveButtons_WaterFor_PCSB[Index]
            ValveButton_FlourFor_PCS := AllValveButtons_FlourFor_PCSB[Index]
            ValveButton_CheeseFor_PCS := AllValveButtons_CheeseFor_PCSB[Index]
            #ValveButton_PimentosFor_PCS := AllValveButtons_PimentosFor_PCSB[Index]
            set ButtonsArray = ButtonsArray + array{
                ValveButton_WaterFor_B,
                ValveButton_FlourFor_B,
                ValveButton_WaterFor_CB,
                ValveButton_FlourFor_CB,
                ValveButton_CheeseFor_CB,
                ValveButton_WaterFor_PCS,
                ValveButton_FlourFor_PCS,
                ValveButton_CheeseFor_PCS
                #ValveButton_PimentosFor_PCS
                }
            set SimplifiedButtonsArray = SimplifiedButtonsArray + array{
                Simplified_ValveButtons_B,
                Simplified_ValveButtons_CB,
                Simplified_ValveButtons_PCS
            }
        else:
            #Print("FATAL ERROR: GetValves(): Failed on assigning Buttons Valves. Index {Index}")     
        if:
            #BakerA
            ValveProp_WaterFor_B := AllValveProps_WaterFor_B[Index]
            ValveProp_FlourFor_B := AllValveProps_FlourFor_B[Index]
            #BakerB
            ValveProp_WaterFor_CB := AllValveProps_WaterFor_CB[Index]
            ValveProp_FlourFor_CB := AllValveProps_FlourFor_CB[Index]
            ValveProp_CheeseFor_CB := AllValveProps_CheeseFor_CB[Index]
            #BakerC
            ValveProp_WaterFor_PCS := AllValveProps_WaterFor_PCS[Index]
            ValveProp_FlourFor_PCS := AllValveProps_FlourFor_PCS[Index]
            ValveProp_CheeseFor_PCS := AllValveProps_CheeseFor_PCS[Index]
            ValveProp_PimentosFor_PCS := AllValveProps_PimentosFor_PCS[Index]
            set PropsArray = PropsArray + array{
                ValveProp_WaterFor_B,
                ValveProp_FlourFor_B,
                ValveProp_WaterFor_CB,
                ValveProp_FlourFor_CB,
                ValveProp_CheeseFor_CB,
                ValveProp_WaterFor_PCS,
                ValveProp_FlourFor_PCS,
                ValveProp_CheeseFor_PCS,
                ValveProp_PimentosFor_PCS
                }
        else:
            #Print("FATAL ERROR: GetValves(): Failed on assigning props Valves of Index: {Index}")

        Tuple : tuple([]creative_prop,[]button_device,[]button_device) = {PropsArray, ButtonsArray,SimplifiedButtonsArray}
        return Tuple

<#
    GetProgressBarsProps(Index : int) : []creative_prop =
        var PropsArray : []creative_prop = array{}
        if:
            ProgressBarBread := AllProps_ProgressBar_Bread[Index]
            ProgressBarSlicedBread := AllProps_ProgressBar_SlicedBread[Index]
            ProgressBarCheeseBread := AllProps_ProgressBar_CheeseBread[Index]
            ProgressBarCheeseSalsa := AllProps_ProgressBar_CheeseSalsa[Index]
            ProgressBarPimientoCheeseSandwich := AllProps_ProgressBar_PimientoCheeseSandwich[Index]
            set PropsArray = PropsArray + array{ProgressBarBread,
                ProgressBarSlicedBread,
                ProgressBarCheeseBread, 
                ProgressBarCheeseSalsa, 
                ProgressBarPimientoCheeseSandwich}
        else:
            #Print("FATAL ERROR: GetProgressBarsProps(): Failed on assigning Prop ProgressBars of Index: {Index}")
        return PropsArray#>

    GetFunnelButton(Index : int) : button_device =
        var Button : button_device = button_device{}
        if:
            FunnelButtonBB := AllButtons_Funnel[Index]
        then:
            set Button = FunnelButtonBB
        else:
            #Print("FATAL ERROR: GetFunnelButton(): Failed on assigning button of Index: {Index}")
        return Button  


    GetUpgradesButtons(Index : int) : []button_device =
        var ButtonArray : []button_device = array{}
        if:
            #Baker upgrades
            UpgradesButtonBB := AllButtons_BB_Upgrades[Index]
            UpgradesButtonCBB := AllButtons_CBB_Upgrades[Index]
            UpgradesButtonPCSB := AllButtons_PCSB_Upgrades[Index]
            #BakerInfo button
            BakerInfoButtonBB := AllButtons_BB_BakerInfo[Index]
            BakerInfoButtonCBB := AllButtons_CBB_BakerInfo[Index]
            BakerInfoButtonPCSB := AllButtons_PCSB_BakerInfo[Index]
            #Deposit button
            UpgradesButtonDeposit := AllButtons_UpgradeDeposit[Index]
            #Upgrade drone
            UpgradeDroneButton := AllButtons_UpgradeDrone[Index]
            #Ship drone
            ShipDroneButton := AllButtons_ShipDrone[Index]
            
            set ButtonArray = ButtonArray + array{UpgradesButtonBB,
            UpgradesButtonCBB,
            UpgradesButtonPCSB,
            BakerInfoButtonBB,
            BakerInfoButtonCBB,
            BakerInfoButtonPCSB,
            UpgradesButtonDeposit,
            UpgradeDroneButton,
            ShipDroneButton}
        else:
            #Print("FATAL ERROR: GetComputerButtons(): Failed on assigning computers buttons of Index: {Index}")

        return ButtonArray

    GetBakerInfoButtons(Index : int) : []button_device =
        var ButtonArray : []button_device = array{}
        if:
            BakerInfoButtonBB := AllButtons_BB_BakerInfo[Index]
            BakerInfoButtonCBB := AllButtons_CBB_BakerInfo[Index]
            BakerInfoButtonPCSB:= AllButtons_PCSB_BakerInfo[Index]
            set ButtonArray = ButtonArray + array{BakerInfoButtonBB,
            BakerInfoButtonCBB,
            BakerInfoButtonPCSB}
        else:
            #Print("FATAL ERROR: GetComputerButtons(): Failed on assigning computers buttons of Index: {Index}")

        return ButtonArray

    GetSuperEnergyDevices(Index : int) : tuple(timer_device,creative_prop) =
        var TimerDevice : timer_device = timer_device{}
        var CreativeProp : creative_prop = creative_prop{}
        if:
            ATimerDevice := AllTimerDevices[Index]
            AOverPoweredSign := AllOverPoweredSigns[Index]
        then:
            set TimerDevice = ATimerDevice
            set CreativeProp = AOverPoweredSign
        else:
            PrintFatalError("On GetSuperEnergyDevices")
        return (TimerDevice,CreativeProp)


        <#
    GetPopDialogs(Index : int) : []popup_dialog_device =
        var PopDialogArray : []popup_dialog_device = array{}
        if:
            Dialog := AllPopDialogDevices_ComputerInfo[Index]
        then:
            set PopDialogArray = PopDialogArray + array{
                Dialog
            }
        return PopDialogArray#>

    GetMainDepositsProps(Index : int) : []creative_prop =
        var PropsArray : []creative_prop = array{}
        if:
            MainWaterDeposit_AmmountProp := AllMainWaterDeposit_AmmountProp[Index]
            MainFlourDeposit_AmmountProp := AllMainFlourDeposit_AmmountProp[Index]
            MainCheeseDeposit_AmmountProp := AllMainCheeseDeposit_AmmountProp[Index]
            MainPimentosDeposit_AmmountProp := AllMainPimentosDeposit_AmmountProp[Index]
        then:
            set PropsArray = PropsArray + array{
                MainWaterDeposit_AmmountProp,
                MainFlourDeposit_AmmountProp,
                MainCheeseDeposit_AmmountProp, 
                MainPimentosDeposit_AmmountProp}
        else:
            #Print("FATAL ERROR: GetMainDepositsProps(): Failed on assigning Prop Deposits of Index: {Index}")
        return PropsArray
    
    GetBillboards(Index : int) : []billboard_device =
        var BillBoardArray : []billboard_device = array{}
        if:
            WaterDeposit2 := AllBillboards_WaterDeposit[Index]
            FlourDeposit := AllBillboards_FlourDeposit[Index]
            CheeseDeposit := AllBillboards_CheeseDeposit[Index]
            PimentosDeposit := AllBillboards_PimentosDeposit[Index]
            DroneDeposit := AllBillboards_DroneDeposit[Index]
            DroneShip := AllBillboards_DroneShip[Index]
        then:
            set BillBoardArray = array{WaterDeposit2,FlourDeposit,CheeseDeposit,PimentosDeposit,DroneDeposit,DroneShip}
        return BillBoardArray

    GetBakersProps(Index : int) : []creative_prop =
        var PropsArray : []creative_prop = array{}
        for(CBB_Prop : AllCBBProps):
            CBB_Prop.SetMaterial(Materials.MI_BarrierBlue,?Index:=0)
            CBB_Prop.SetMaterial(Materials.MI_BarrierBlue,?Index:=1)
            CBB_Prop.SetMaterial(Materials.MI_BarrierBlue,?Index:=2)
            CBB_Prop.SetMaterial(Materials.MI_BarrierBlue,?Index:=3)
        for(PCSB_Prop : AllPCSBProps):
            PCSB_Prop.SetMaterial(Materials.MI_BarrierBlue,?Index:=0)
            PCSB_Prop.SetMaterial(Materials.MI_BarrierBlue,?Index:=1)
            PCSB_Prop.SetMaterial(Materials.MI_BarrierBlue,?Index:=2)
            PCSB_Prop.SetMaterial(Materials.MI_BarrierBlue,?Index:=3)
        if:
            BBProp := AllBBProps[Index]
            CBBProp := AllCBBProps[Index]
            PCSBProp := AllPCSBProps[Index]
            DroneProp := AllDroneProps[Index]
        then:
            set PropsArray = PropsArray + array{
                BBProp,
                CBBProp,
                PCSBProp,
                DroneProp}
        else:
            #Print("FATAL ERROR: GetDepositsProps(): Failed on assigning Prop Deposits of Index: {Index}")
        return PropsArray

    LoadBakeryProps(TestProp : creative_prop, Transform : transform, AllBakeryUpgrades : []int) : void =
        spawn:
            AnimatePropAppearance(TestProp,Transform,false)
        #Print("Trying to load props")



  