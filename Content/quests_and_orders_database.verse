using { /UnrealEngine.com/Temporary/UI }
using { /UnrealEngine.com/Temporary/SpatialMath }
using { /Verse.org/Assets }
using { /Verse.org/Simulation }
using { /Verse.org/Random }
using { /Fortnite.com/Devices }
using { /Fortnite.com/Characters }

#All challenges DataBase

TutorialCheese_Index : int = 0
TutorialPimentos_Index : int = 1
TutorialMutantMaterial_Index : int = 2
TutorialCoal_Index : int = 3
TutorialOrder_Index : int = 32
TutorialPet_Index : int = 5
TutorialDungeon_Index : int = 6
TutorialHexaGame_Index : int = 7
TutorialAggedSapphires_Index : int = 8
Quest0_Index : int = 34

MAIN_LORE : int = 0
G_RANDOM_LORE : int = 1
W_RANDOM_LORE : int = 2

J0_RANDOM_LORE : int = 3
J1_RANDOM_LORE : int = 4
H_RANDOM_LORE : int = 5
J2_RANDOM_LORE : int = 6

quests_and_orders_database := class(creative_device):

    MASTERYSPECIALQUEST_INDEX : int = 14

    #G_CHARACTER
    AllSecretsTips : []string = array{
        "Hi! I'm Mike, the bakery assistant"
        "Talk to me for help and advice!"
        "Teamwork boosts efficiency by sharing Coal energy!"
        "The longer you hit Stale Bread, the more Mutant Material!"
        "Teleporting from the Bakery is always free!"
        "You can block all Mutant Bread attacks with your melee weapon!"
        "Seen hidden targets? Try shooting them with a bow!"
        "Add Coal from your Bakery using the available Funnel!"
        "Always upgrade your Sourdough with Coal!"
        "Helping in any Dungeon gives everyone full rewards"
        "Your Sourdough Challenge setup is always saved!"
        "Repair symbol PCs give rewards—help factory upkeep!"
        "Save up Sourdough Challenge tokens for better use!"
    }


    AllSecretsHelp : []string = array{
        "I think I saw Special Data near the factory" #SECRET 0
        "I saw Special Data in the junkyard crusher" #SECRET 1
        "I saw Special Data at the vacuum shop" #SECRET 2
        "" #SECRET 3
        "I'm sure I saw Special Data in the restaurant" #SECRET 4
        "Steve left Special Data at the car wash" #SECRET 5
        "" #SECRET 6
        "" #SECRET 7
        "I saw Special Data in the chimney of the pool house" #SECRET 8
        "" #SECRET 9
        "" #SECRET 10
        "I found Special Data under a Stale Bread in the desert" #SECRET 11
        "" # SECRET 12
        "" # SECRET 13
        "" # SECRET 14
        "" # SECRET 15
        "" # SECRET 16
        "" # SECRET 17
        "" # SECRET 18
        "" # SECRET 19
        "" # SECRET 20
        "" # SECRET 21
    }


    #W_CHARACTER
    AllWLore : []string = array{
        <#LORE0#> "Welcome to the factory. Try not to make it worse" #BIENVENIDA
        <#LORE1#> "Steve, my last apprentice, ruined everything!" #BIENVENIDA
        <#LORE2#> "" #ZONE 1
        <#LORE3#> "The Stale Bread infestations? All Steve’s doing" #ZONE 2
        <#LORE4#> "Steve’s junked ideas belong here. Right where they failed!" #ZONE 3
        <#LORE5#> "Laundry? Can’t wash away the mess Steve made!" #ZONE4
        <#LORE6#> "Last place Steve was seen… He ran instead of fixing this!" #ZONE 5
        <#LORE7#> "Steve wasted time here too? What was he even doing?" #ZONE 6
        <#LORE8#> "Leaving town? Hah! Maybe you’ll find Steve out there!" #ZONE 7
        <#LORE9#> "Steve ran off to the desert? Hah, coward!" #ZONE 8
        <#LORE10#> "He buried Mutant Bread? A lazy cover-up!" #ZONE 9
        <#LORE11#> "If he makes real bread, I’ll eat my own hat!" #ZONE 10
        <#LORE12#> "He thinks he’s onto something? Laughable" #ZONE 11 (DON ELADIO)
        <#LORE13#> "Now he’s experimenting? Just another failure!" #ZONE 12
        <#LORE14#> "Hah! Blaming others won’t fix your failure!" #ZONE 13
        <#LORE15#> "Hah! Another factory? You’ll fail again, just like before!" #ZONE 14
        <#LORE16#> "Hah! Another failure waiting to happen!"#ZONE 15 ENTRADA NEW FACTORY
        <#LORE17#> "Mutant Bread caging experiments? He’s lost it" #ZONE 16 (CAGES)
        <#LORE18#> "Lies! You failed yourself, don’t blame me!" #ZONE 17
        <#LORE19#> "He never had what it takes to reach my level" #ZONE 18
        <#LORE20#> "96.2%? That’s cute. Mine is the REAL deal" #ZONE 19 
        <#LORE21#> "I didn’t need help to perfect MY formula" #STEVE ASK FOR HELP
        <#LORE22#> "Mutant bread?! That fool really outdid himself!" #FIRST ENCOUNTER
    }

    AllWRandom : []string = array{
        "You better not mess up like Steve did"
        "Clean that up! This isn’t a playground"
        "Steve tried ‘improving’ my work... Look at this mess"
        "Tired? Ha! Try running a factory alone!"
        "You listening? Then get back to work!"
        "Coal’s expensive. Don’t waste it like Steve did."
        "Fixing this mess takes time. Unlike Steve, I try."
        "Mutant bread is YOUR problem now, not mine."
        "Hard bread? Steve’s ‘gift’ to the city..."
        "You think you can bake better? Prove it!"
    }

    #J_CHARACTER
    AllJLore : []string = array{
        <#LORE0#> "" #ZONE WELCOME
        <#LORE1#> ""#ZONE 0
        <#LORE2#> ""#ZONE 1
        <#LORE3#> ""#ZONE 2
        <#LORE4#> ""#ZONE 3
        <#LORE5#> ""#ZONE 4
        <#LORE6#> ""#ZONE 5
        <#LORE7#> ""#ZONE 6
        <#LORE8#> ""#ZONE 7
        <#LORE9#> "You found me? Heh, I didn’t think anyone would" #ZONE 8
        <#LORE10#> "I tried burying the bread… It didn’t work" #ZONE 9
        <#LORE11#> "My formula was right. Something went wrong" #ZONE 10
        <#LORE12#> "A friend told me to check the old factory" #ZONE 11 (DON ELADIO)
        <#LORE13#> "I’ll test my formula at the old factory. Something’s off" #ZONE 12
        <#LORE14#> "My formula was perfect… Someone sabotaged it!" #ZONE 13
        <#LORE15#> "I built a new factory. I’ll fix this mess"  #ZONE 14 SECRET BAKERY
        <#LORE16#> "This is it. My factory. The real formula starts here" #ZONE 15 ENTRANCE
        <#LORE17#> "I caged the mutant bread. Studying it now"  #ZONE 16 (CAGES)
        <#LORE18#> "I found it! It was a setup… He made sure I’d fail!"  #ZONE 17
        <#LORE19#> "I’m close… The formula is nearly pure!" #ZONE 18 
        <#LORE20#> "96.2% purity… I finally did it!" #ZONE 19
    }
        
    AllJRandom : []string = array{
        "My RV broke down. Guess I’m walking now"
        "I had to leave. The bread spread too fast"
        "Mutant bread… It still haunts me"
        "You ever seen bread fight back? I have"
        "I miss my RV. Walking sucks"
        "I need more Coal. Got any?"
        "I swear, I almost cracked the formula!"
        "Sleep? Haven’t done that in weeks"
        "This desert is too damn hot"
        "One day, I’ll prove I was right"
    }

    AllJRandom_SecretBakery : []string = array{
        "Help me stabilize the formula on the Bakery PCs!" #FIRST IN TERACTION
        "This factory… It’s my fresh start"
        "Still here? Good, I need the help"
        "Pure bread isn’t just a dream anymore"
        "Mutant bread? No. Only perfection now"
        "He would never admit I was right"
        "Baking takes patience. Like real science"
        "I barely sleep, but this work drives me"
        "Coal’s low again? Always need more fuel"
        "Every batch gets better. We’re so close"
        "Mutant bread… what a nightmare that was"
        "Nothing explodes anymore. That’s progress"
        "I don’t need luck. Just the right formula"
        "Sourdough isn’t just bread. It’s proof"
        "I’d offer you bread, but it’s all research"
        "Perfection isn’t easy, but I’ll get there"
    }

    AllJRandom_Pure : []string = array{
        "96.2%... Close, but not enough"
        "He still holds the top spot… for now"
        "I can feel it… Perfection is within reach"
        "He never wanted me to get this far"
        "He’s still ahead… But not for long"
        "This isn’t over. Not until I reach 99%"
        "You helped me once… Will you help again?"
        "I proved my formula works. I just need more"
        "96.2% purity… So frustratingly close!"
        "I need to push further"
        "He sabotaged me once. He won’t stop me again"
        "If we work together, we might surpass him"
        "Jonny still stands above me. That has to change"
        "Sourdough is stable… but not perfect"
        "I won’t stop until I see true purity"
        "This is just the beginning, not the end"
    }


    #H_CHARACTER
    AllHRandom : []string = array{
        "Well, well… Look what we got here"
        "You know who owns these cars?"
        "This bread mess just keeps getting weirder"
        "Steve? Yeah, I heard about that guy"
        "Mutant bread… Hah! What a joke"
        "So, you’re involved in this, huh?"
        "What’s Steve running from?"
        "You sure you wanna dig deeper?"
        "I’ve seen worse… but not by much"
        "Something stinks, and it ain’t just the bread"
        "You looking for answers? So am I"
        "City’s getting restless. Too much weird stuff"
        "If Jonny’s hiding something, I’ll find out"
        "This whole case is a mess"
        "Stay sharp. Something doesn’t add up"
        "Last time I trusted a baker, I got burned"
    }



    #Lore
    @editable
    AllLoreCameras : []gameplay_camera_fixed_point_device = array{}
    @editable
    AllEnabledLoreCameras : []logic = array{}
    var AllAgentLoreAccumulated : []int = array{0,0,0,0}
    @editable
    AllWLoreRandom_Buttons : button_device = button_device{}
    @editable
    AllHLoreRandom_Button : button_device = button_device{}
    @editable
    AllGLoreRandom_Button : button_device = button_device{}

    @editable
    AllJLoreRandomRV_Button : button_device = button_device{}
    @editable
    AllJLoreRandomBaker_Button : button_device = button_device{}
    @editable
    AllJLoreRandomPure_Button : button_device = button_device{}

    #Orders
    @editable
    NumMaxPicks_PerOrder : float = 200.0
    @editable
    MaxCoal_PerOrder : int = 30

    #Baker Tutorial
    var AllAgentsWatchingTutorial : []agent = array{}
    @editable
    BakerTutorialProp : creative_prop = creative_prop{}
    @editable
    ValveTutorialProp0 : creative_prop = creative_prop{}
    @editable
    ValveTutorialProp1 : creative_prop = creative_prop{}
    @editable
    PlayBakerTutorialButton : button_device = button_device{}
    @editable
    AllBakerTutorialButtons : []button_device = array{}
    @editable
    AllBakerTutorialCameras : []gameplay_camera_fixed_point_device = array{}

    #Pet Tutorial
    @editable
    AllADCPetButtons : []button_device = array{}
    @editable
    AllSupportPetButtons : []button_device = array{}
    @editable
    AllTankPetButtons : []button_device = array{}
    @editable
    AllFighterPetButtons : []button_device = array{}
    @editable
    AllAssasinPetButtons : []button_device = array{}

    @editable
    AllTutorialCreativeProps : []creative_prop = array{}
    @editable
    AllTutorialButtonDevices : []button_device = array{}
    @editable
    AllQuestCreativeProps : []creative_prop = array{}
    @editable
    AllQuestButtons : []button_device = array{}

    var MaybeGameManager : ?game_manager = false

    StandarQuestReward_Coal : float = 30.0

    PlayersDataManager : player_data_manager = player_data_manager{}

    var AllQuestInitiated : logic = false
    var AllTutorialQuests : []tuple(string,texture,string,int,float,int,float,?creative_prop,?button_device)= array{}
    var AllNormalQuests : []tuple(string,texture,string,int,float,int,float,?creative_prop,?button_device)= array{}   
    var AllSuperSpecialQuests : []tuple(string,texture,string,int,float,int,float,?creative_prop,?button_device)= array{}   
    
    OnBegin<override>()<suspends>:void=

        AllWLoreRandom_Buttons.InteractedWithEvent.Subscribe(handler_random{MaybeQuestAndOrders:=option{Self}, Index:=W_RANDOM_LORE}.HandlerLoreInteraction)
        AllHLoreRandom_Button.InteractedWithEvent.Subscribe(handler_random{MaybeQuestAndOrders:=option{Self}, Index:=H_RANDOM_LORE}.HandlerLoreInteraction)
        AllGLoreRandom_Button.InteractedWithEvent.Subscribe(handler_random{MaybeQuestAndOrders:=option{Self}, Index:=G_RANDOM_LORE}.HandlerLoreInteraction)
        AllJLoreRandomRV_Button.InteractedWithEvent.Subscribe(handler_random{MaybeQuestAndOrders:=option{Self}, Index:=J0_RANDOM_LORE}.HandlerLoreInteraction)
        AllJLoreRandomBaker_Button.InteractedWithEvent.Subscribe(handler_random{MaybeQuestAndOrders:=option{Self}, Index:=J1_RANDOM_LORE}.HandlerLoreInteraction)
        AllJLoreRandomPure_Button.InteractedWithEvent.Subscribe(handler_random{MaybeQuestAndOrders:=option{Self}, Index:=J2_RANDOM_LORE}.HandlerLoreInteraction)
        
        CheckQuestInited()

        PlayBakerTutorialButton.InteractedWithEvent.Subscribe(HandlerPlayBakerTutorial)

        for(Button : AllBakerTutorialButtons):
            Button.InteractedWithEvent.Subscribe(HandlerPlayBakerTutorial)

        for(Button : AllADCPetButtons):
            Button.InteractedWithEvent.Subscribe(DisplayADCTutorial)
            
        for(Button : AllSupportPetButtons):
            Button.InteractedWithEvent.Subscribe(DisplaySupportTutorial)

        for(Button : AllTankPetButtons):
            Button.InteractedWithEvent.Subscribe(DisplayTankTutorial)

        for(Button : AllFighterPetButtons):
            Button.InteractedWithEvent.Subscribe(DisplayFighterTutorial)

        for(Button : AllAssasinPetButtons):
            Button.InteractedWithEvent.Subscribe(DisplayAssasinTutorial)
        

    DisplayADCTutorial(Agent : agent) : void =
        if(GameManager:=MaybeGameManager?):
            GameManager.DisplayCharacterMessage(Agent,I_CHARACTER,"Meet the ADC Sourdough",?Sound:=MESSAGE_SOUND,?DisplayTime:=3.0)
            GameManager.DisplayCharacterMessage(Agent,I_CHARACTER,"It produces arrows over time and boosts Bow damage",?Sound:=MESSAGE_SOUND,?DisplayTime:=5.0)
            GameManager.DisplayCharacterMessage(Agent,I_CHARACTER,"In Sourdough Competition, it attacks from long range",?Sound:=MESSAGE_SOUND,?DisplayTime:=5.0)
            spawn:
                DelayTryAddPetTutorial(Agent)

    DisplaySupportTutorial(Agent : agent) : void =
        if(GameManager:=MaybeGameManager?):
            GameManager.DisplayCharacterMessage(Agent,G_CHARACTER,"Meet the Support Sourdough",?Sound:=MESSAGE_SOUND,?DisplayTime:=3.0)
            GameManager.DisplayCharacterMessage(Agent,G_CHARACTER,"It heals you over time and has an emergency heal",?Sound:=MESSAGE_SOUND,?DisplayTime:=5.0)
            GameManager.DisplayCharacterMessage(Agent,G_CHARACTER,"In Sourdough Competition, it can heal allies",?Sound:=MESSAGE_SOUND,?DisplayTime:=5.0)
            spawn:
                DelayTryAddPetTutorial(Agent)

    DisplayTankTutorial(Agent : agent) : void =
        if(GameManager:=MaybeGameManager?):
            GameManager.DisplayCharacterMessage(Agent,G_CHARACTER,"Meet the Tank Sourdough",?Sound:=MESSAGE_SOUND,?DisplayTime:=3.0)
            GameManager.DisplayCharacterMessage(Agent,G_CHARACTER,"It increases your max health and grants an emergency shield",?Sound:=MESSAGE_SOUND,?DisplayTime:=5.0)
            GameManager.DisplayCharacterMessage(Agent,G_CHARACTER,"In Sourdough Competition, it has high health",?Sound:=MESSAGE_SOUND,?DisplayTime:=5.0)
            spawn:
                DelayTryAddPetTutorial(Agent)

    DisplayFighterTutorial(Agent : agent) : void =
        if(GameManager:=MaybeGameManager?):
            GameManager.DisplayCharacterMessage(Agent,G_CHARACTER,"Meet the Fighter Sourdough",?Sound:=MESSAGE_SOUND,?DisplayTime:=3.0)
            GameManager.DisplayCharacterMessage(Agent,G_CHARACTER,"It boosts your melee damage and grants a melee Super Strike",?Sound:=MESSAGE_SOUND,?DisplayTime:=5.0)
            GameManager.DisplayCharacterMessage(Agent,G_CHARACTER,"In Sourdough Competition, it has solid damage and health",?Sound:=MESSAGE_SOUND,?DisplayTime:=5.0)
            spawn:
                DelayTryAddPetTutorial(Agent)

    DisplayAssasinTutorial(Agent : agent) : void =
        if(GameManager:=MaybeGameManager?):
            GameManager.DisplayCharacterMessage(Agent,G_CHARACTER,"Meet the Hunter Sourdough",?Sound:=MESSAGE_SOUND,?DisplayTime:=3.0)
            GameManager.DisplayCharacterMessage(Agent,G_CHARACTER,"It boosts Mutants find rate and grants extra weapon critical",?Sound:=MESSAGE_SOUND,?DisplayTime:=5.0)
            GameManager.DisplayCharacterMessage(Agent,G_CHARACTER,"In Sourdough Competition, it jumps opposite at start",?Sound:=MESSAGE_SOUND,?DisplayTime:=5.0)
            spawn:
                DelayTryAddPetTutorial(Agent)
          
    DelayTryAddPetTutorial(Agent : agent)<suspends> : void =
        Sleep(15.0)
        if(GameManager:=MaybeGameManager?, ValidBakery := GameManager.GetAgentBakery_Transacted(Agent)?, Player := player[Agent]):
            UpdatedDataTable := GameManager.QuestsAndOrdersDatabase.AddAgentQuest_ProgressV2(Player,ValidBakery.PlayerDataCache,1.0,PETTUTORIAL_TUTORIAL_TYPE)(4)
            ValidBakery.SetNewPlayerDataCache(UpdatedDataTable)

    HandlerPlayBakerTutorial(Agent : agent) : void =
        var IsAlreadyWatching : logic= false
        for(Index:=0..AllAgentsWatchingTutorial.Length-1, AgentTutorial := AllAgentsWatchingTutorial[Index]):
            if(Agent = AgentTutorial):
                set IsAlreadyWatching = true

        if (not IsAlreadyWatching?):
            set AllAgentsWatchingTutorial = AllAgentsWatchingTutorial + array{Agent}
            spawn:
                PlayBakerTutorial(Agent)

    DisplayBakerDummyMessage(Agent : agent) : void= 
        if(GameManager:=MaybeGameManager?):
            GameManager.DisplayCharacterMessage(Agent,G_CHARACTER,"This is a Dummy Bakery! Apply the tutorial lessons to your Bakery",?Sound:=MESSAGE_SOUND,?DisplayTime:=5.0)


    PlayBakerTutorial(Agent : agent)<suspends> : void =
        if:
            GameManager:=MaybeGameManager?
            FortCharacter := Agent.GetFortCharacter[]
            PlayerBakery := GameManager.GetAgentBakery_Transacted(Agent)?
            Camera0:=AllBakerTutorialCameras[0]
            Camera1:=AllBakerTutorialCameras[1]
            Camera2:=AllBakerTutorialCameras[2]
            Camera3:=AllBakerTutorialCameras[3]
            Camera4:=AllBakerTutorialCameras[4]
            Camera5:=AllBakerTutorialCameras[5]
            Camera6:=AllBakerTutorialCameras[6]
            Camera7:=AllBakerTutorialCameras[7]
            Camera8:=AllBakerTutorialCameras[8]
            Camera9:=AllBakerTutorialCameras[9]
            MainInventoryUI := PlayerBakery.MaybeMainInventoryUI?
        then:
            FortCharacter.PutInStasis(stasis_args{AllowEmotes:=false,AllowFalling:=false,AllowTurning:=false})
            MainInventoryUI.Hide_All()
            var DelayTime : float = 3.0  
            GameManager.CleanCharacterMessages(Agent)
            Sleep(0.25)
            #CAMERA 0
            Camera0.AddTo(Agent)
            set DelayTime = 3.0
            GameManager.CleanCharacterMessages(Agent)
            if(PlayerBakery.PlayerDataCache.Language=ENGLISH):
                GameManager.DisplayCharacterMessage(Agent,G_CHARACTER,"Welcome to the Bakery Tutorial!",?DisplayTime:=DelayTime,?Priority:=MAX_PRIORITY<#,?Sound:=MESSAGE_SOUND#>)
            else if(PlayerBakery.PlayerDataCache.Language=SPANISH):
                GameManager.DisplayCharacterMessage(Agent,G_CHARACTER,"¡Bienvenido al tutorial de panadería!",?DisplayTime:=DelayTime,?Priority:=MAX_PRIORITY<#,?Sound:=MESSAGE_SOUND#>)
            else if(PlayerBakery.PlayerDataCache.Language=FRENCH):
                GameManager.DisplayCharacterMessage(Agent,G_CHARACTER,"Bienvenue au tutoriel de boulangerie!",?DisplayTime:=DelayTime,?Priority:=MAX_PRIORITY<#,?Sound:=MESSAGE_SOUND#>)
            else if(PlayerBakery.PlayerDataCache.Language=PORTUGUESE):
                GameManager.DisplayCharacterMessage(Agent,G_CHARACTER,"Bem-vindo ao tutorial de panificação!",?DisplayTime:=DelayTime,?Priority:=MAX_PRIORITY<#,?Sound:=MESSAGE_SOUND#>)            
            Sleep(DelayTime+0.25)
            #CAMERA 1
            Camera1.AddTo(Agent)
            set DelayTime = 5.0
            GameManager.CleanCharacterMessages(Agent)
            if(PlayerBakery.PlayerDataCache.Language=ENGLISH):
                GameManager.DisplayCharacterMessage(Agent,I_CHARACTER,"Ingredients enter when valves are Green or Yellow.",?DisplayTime:=5.0,?Sound:=MESSAGE_SOUND,?Priority:=MAX_PRIORITY)
            else if(PlayerBakery.PlayerDataCache.Language=SPANISH):
                GameManager.DisplayCharacterMessage(Agent,I_CHARACTER,"Los ingredientes entran con válvulas verdes o amarillas.",?DisplayTime:=5.0,?Sound:=MESSAGE_SOUND,?Priority:=MAX_PRIORITY)
            else if(PlayerBakery.PlayerDataCache.Language=FRENCH):
                GameManager.DisplayCharacterMessage(Agent,I_CHARACTER,"Les ingrédients entrent si les vannes sont vertes ou jaunes.",?DisplayTime:=5.0,?Sound:=MESSAGE_SOUND,?Priority:=MAX_PRIORITY)
            else if(PlayerBakery.PlayerDataCache.Language=PORTUGUESE):
                GameManager.DisplayCharacterMessage(Agent,I_CHARACTER,"Os ingredientes entram com válvulas verdes ou amarelas.",?DisplayTime:=5.0,?Sound:=MESSAGE_SOUND,?Priority:=MAX_PRIORITY)            
            Sleep(2.0)
            ValveTutorialProp0.SetMaterial(Materials.Mat_PipeState0, ?Index:=0)
            ValveTutorialProp1.SetMaterial(Materials.Mat_PipeState0, ?Index:=0)
            Sleep(2.0)
            ValveTutorialProp0.SetMaterial(Materials.Mat_PipeState1, ?Index:=0)
            ValveTutorialProp1.SetMaterial(Materials.Mat_PipeState1, ?Index:=0)
            Sleep(2.0)
            ValveTutorialProp0.SetMaterial(Materials.Mat_PipeState2, ?Index:=0)
            ValveTutorialProp1.SetMaterial(Materials.Mat_PipeState2, ?Index:=0)
            GameManager.CleanCharacterMessages(Agent)
            if(PlayerBakery.PlayerDataCache.Language=ENGLISH):
                GameManager.DisplayCharacterMessage(Agent,I_CHARACTER,"A Green valve has more priority than a Yellow one.",?DisplayTime:=5.0,?Sound:=MESSAGE_SOUND,?Priority:=MAX_PRIORITY)
            else if(PlayerBakery.PlayerDataCache.Language=SPANISH):
                GameManager.DisplayCharacterMessage(Agent,I_CHARACTER,"La válvula verde tiene más prioridad que la amarilla.",?DisplayTime:=5.0,?Sound:=MESSAGE_SOUND,?Priority:=MAX_PRIORITY)
            else if(PlayerBakery.PlayerDataCache.Language=FRENCH):
                GameManager.DisplayCharacterMessage(Agent,I_CHARACTER,"La vanne verte a plus de priorité que la jaune.",?DisplayTime:=5.0,?Sound:=MESSAGE_SOUND,?Priority:=MAX_PRIORITY)
            else if(PlayerBakery.PlayerDataCache.Language=PORTUGUESE):
                GameManager.DisplayCharacterMessage(Agent,I_CHARACTER,"A válvula verde tem mais prioridade que a amarela.",?DisplayTime:=5.0,?Sound:=MESSAGE_SOUND,?Priority:=MAX_PRIORITY)            
            Sleep(2.0)
            ValveTutorialProp0.SetMaterial(Materials.Mat_PipeState1, ?Index:=0)
            ValveTutorialProp1.SetMaterial(Materials.Mat_PipeState1, ?Index:=0)
            Sleep(2.0)
            ValveTutorialProp0.SetMaterial(Materials.Mat_PipeState2, ?Index:=0)
            ValveTutorialProp1.SetMaterial(Materials.Mat_PipeState2, ?Index:=0)
            Sleep(1.0)
            #CAMERA 2
            Camera2.AddTo(Agent)
            set DelayTime = 5.0
            GameManager.CleanCharacterMessages(Agent)
            if(PlayerBakery.PlayerDataCache.Language=ENGLISH):
                GameManager.DisplayCharacterMessage(Agent,I_CHARACTER,"This bar shows what’s needed for the next batch.",?DisplayTime:=5.0,?Sound:=MESSAGE_SOUND,?Priority:=MAX_PRIORITY)
            else if(PlayerBakery.PlayerDataCache.Language=SPANISH):
                GameManager.DisplayCharacterMessage(Agent,I_CHARACTER,"Esta barra indica lo necesario para el próximo lote.",?DisplayTime:=5.0,?Sound:=MESSAGE_SOUND,?Priority:=MAX_PRIORITY)
            else if(PlayerBakery.PlayerDataCache.Language=FRENCH):
                GameManager.DisplayCharacterMessage(Agent,I_CHARACTER,"Cette barre montre le besoin pour le prochain lot.",?DisplayTime:=5.0,?Sound:=MESSAGE_SOUND,?Priority:=MAX_PRIORITY)
            else if(PlayerBakery.PlayerDataCache.Language=PORTUGUESE):
                GameManager.DisplayCharacterMessage(Agent,I_CHARACTER,"Esta barra indica o necessário para o próximo lote.",?DisplayTime:=5.0,?Sound:=MESSAGE_SOUND,?Priority:=MAX_PRIORITY)            
            for(Index:=2..5):
                ValveTutorialProp0.SetMaterial(Materials.RedScreen_Inst, ?Index:=Index)
                ValveTutorialProp1.SetMaterial(Materials.RedScreen_Inst, ?Index:=Index)
            Sleep(1.0)
            for(Index:=2..5):
                ValveTutorialProp0.SetMaterial(Materials.GreenScreen_Inst, ?Index:=Index)
                ValveTutorialProp1.SetMaterial(Materials.GreenScreen_Inst, ?Index:=Index)
                Sleep(1.0)
            Sleep(0.25)
            #CAMERA 3
            Camera3.AddTo(Agent)
            set DelayTime = 5.0
            GameManager.CleanCharacterMessages(Agent)
            if(PlayerBakery.PlayerDataCache.Language=ENGLISH):
                GameManager.DisplayCharacterMessage(Agent,I_CHARACTER,"Check Baker info on the Information computer.",?DisplayTime:=DelayTime,?Sound:=MESSAGE_SOUND,?Priority:=MAX_PRIORITY)
            else if(PlayerBakery.PlayerDataCache.Language=SPANISH):
                GameManager.DisplayCharacterMessage(Agent,I_CHARACTER,"Consulta info del panificador en el ordenador.",?DisplayTime:=DelayTime,?Sound:=MESSAGE_SOUND,?Priority:=MAX_PRIORITY)
            else if(PlayerBakery.PlayerDataCache.Language=FRENCH):
                GameManager.DisplayCharacterMessage(Agent,I_CHARACTER,"Vérifiez l'info du four sur l'ordinateur.",?DisplayTime:=DelayTime,?Sound:=MESSAGE_SOUND,?Priority:=MAX_PRIORITY)
            else if(PlayerBakery.PlayerDataCache.Language=PORTUGUESE):
                GameManager.DisplayCharacterMessage(Agent,I_CHARACTER,"Veja info do padeiro no computador.",?DisplayTime:=DelayTime,?Sound:=MESSAGE_SOUND,?Priority:=MAX_PRIORITY)            
            Sleep(DelayTime+0.25)
            #CAMERA 4
            Camera4.AddTo(Agent)
            set DelayTime = 5.0
            GameManager.CleanCharacterMessages(Agent)
            if(PlayerBakery.PlayerDataCache.Language=ENGLISH):
                GameManager.DisplayCharacterMessage(Agent,I_CHARACTER,"If the Baker is working, the light turns green.",?DisplayTime:=DelayTime,?Sound:=MESSAGE_SOUND,?Priority:=MAX_PRIORITY)
            else if(PlayerBakery.PlayerDataCache.Language=SPANISH):
                GameManager.DisplayCharacterMessage(Agent,I_CHARACTER,"Si el panificador trabaja, la luz se pone verde.",?DisplayTime:=DelayTime,?Sound:=MESSAGE_SOUND,?Priority:=MAX_PRIORITY)
            else if(PlayerBakery.PlayerDataCache.Language=FRENCH):
                GameManager.DisplayCharacterMessage(Agent,I_CHARACTER,"Si le boulanger travaille, la lumière est verte.",?DisplayTime:=DelayTime,?Sound:=MESSAGE_SOUND,?Priority:=MAX_PRIORITY)
            else if(PlayerBakery.PlayerDataCache.Language=PORTUGUESE):
                GameManager.DisplayCharacterMessage(Agent,I_CHARACTER,"Se o padeiro trabalha, a luz fica verde.",?DisplayTime:=DelayTime,?Sound:=MESSAGE_SOUND,?Priority:=MAX_PRIORITY)            
            Sleep(2.0)
            BakerTutorialProp.SetMaterial(Materials.MI_Complement_green, ?Index:=1)
            Sleep(2.0)
            Sleep(1.25)
            #CAMERA 5
            Camera5.AddTo(Agent)
            set DelayTime = 5.0
            GameManager.CleanCharacterMessages(Agent)
            if(PlayerBakery.PlayerDataCache.Language=ENGLISH):
                GameManager.DisplayCharacterMessage(Agent,I_CHARACTER,"You can also check your next batch inside the Baker.",?DisplayTime:=DelayTime,?Sound:=MESSAGE_SOUND,?Priority:=MAX_PRIORITY)
            else if(PlayerBakery.PlayerDataCache.Language=SPANISH):
                GameManager.DisplayCharacterMessage(Agent,I_CHARACTER,"Verifica tu próximo lote dentro del panificador.",?DisplayTime:=DelayTime,?Sound:=MESSAGE_SOUND,?Priority:=MAX_PRIORITY)
            else if(PlayerBakery.PlayerDataCache.Language=FRENCH):
                GameManager.DisplayCharacterMessage(Agent,I_CHARACTER,"Vérifiez votre prochain lot dans le four.",?DisplayTime:=DelayTime,?Sound:=MESSAGE_SOUND,?Priority:=MAX_PRIORITY)
            else if(PlayerBakery.PlayerDataCache.Language=PORTUGUESE):
                GameManager.DisplayCharacterMessage(Agent,I_CHARACTER,"Verifique seu próximo lote dentro do padeiro.",?DisplayTime:=DelayTime,?Sound:=MESSAGE_SOUND,?Priority:=MAX_PRIORITY)            
            
            spawn:
                GameManager.VFXManager.AnimateBake(BakerTutorialProp.GetTransform(),DelayTime,false,B_TYPE)
            Sleep(DelayTime+0.25)
            #CAMERA 6
            Camera6.AddTo(Agent)
            set DelayTime = 5.0
            GameManager.CleanCharacterMessages(Agent)
            if(PlayerBakery.PlayerDataCache.Language=ENGLISH):
                GameManager.DisplayCharacterMessage(Agent,I_CHARACTER,"Upgrade the Baker’s production at this computer.",?DisplayTime:=3.0,?Sound:=MESSAGE_SOUND,?Priority:=MAX_PRIORITY)
            else if(PlayerBakery.PlayerDataCache.Language=SPANISH):
                GameManager.DisplayCharacterMessage(Agent,I_CHARACTER,"Mejora la producción del panificador aquí.",?DisplayTime:=3.0,?Sound:=MESSAGE_SOUND,?Priority:=MAX_PRIORITY)
            else if(PlayerBakery.PlayerDataCache.Language=FRENCH):
                GameManager.DisplayCharacterMessage(Agent,I_CHARACTER,"Améliorez la production du boulanger ici.",?DisplayTime:=3.0,?Sound:=MESSAGE_SOUND,?Priority:=MAX_PRIORITY)
            else if(PlayerBakery.PlayerDataCache.Language=PORTUGUESE):
                GameManager.DisplayCharacterMessage(Agent,I_CHARACTER,"Melhore a produção do padeiro aqui.",?DisplayTime:=3.0,?Sound:=MESSAGE_SOUND,?Priority:=MAX_PRIORITY)            
            Sleep(3.25)
            GameManager.CleanCharacterMessages(Agent)
            if(PlayerBakery.PlayerDataCache.Language=ENGLISH):
                GameManager.DisplayCharacterMessage(Agent,I_CHARACTER,"Upgrading production INCREASES required ingredients.",?DisplayTime:=3.5,?Sound:=MESSAGE_SOUND,?Priority:=MAX_PRIORITY)
            else if(PlayerBakery.PlayerDataCache.Language=SPANISH):
                GameManager.DisplayCharacterMessage(Agent,I_CHARACTER,"Mejorar la producción aumenta los ingredientes.",?DisplayTime:=3.5,?Sound:=MESSAGE_SOUND,?Priority:=MAX_PRIORITY)
            else if(PlayerBakery.PlayerDataCache.Language=FRENCH):
                GameManager.DisplayCharacterMessage(Agent,I_CHARACTER,"Améliorer la production augmente les ingrédients.",?DisplayTime:=3.5,?Sound:=MESSAGE_SOUND,?Priority:=MAX_PRIORITY)
            else if(PlayerBakery.PlayerDataCache.Language=PORTUGUESE):
                GameManager.DisplayCharacterMessage(Agent,I_CHARACTER,"Melhorar a produção aumenta os ingredientes.",?DisplayTime:=3.5,?Sound:=MESSAGE_SOUND,?Priority:=MAX_PRIORITY)
            Sleep(3.75)
            #CAMERA 7
            Camera7.AddTo(Agent)
            set DelayTime = 5.0
            GameManager.CleanCharacterMessages(Agent)
            if(PlayerBakery.PlayerDataCache.Language=ENGLISH):
                GameManager.DisplayCharacterMessage(Agent,I_CHARACTER,"Your unused ingredients are stored in the Main Deposits.",?DisplayTime:=DelayTime,?Sound:=MESSAGE_SOUND,?Priority:=MAX_PRIORITY)
            else if(PlayerBakery.PlayerDataCache.Language=SPANISH):
                GameManager.DisplayCharacterMessage(Agent,I_CHARACTER,"Los ingredientes no usados están en los depósitos.",?DisplayTime:=DelayTime,?Sound:=MESSAGE_SOUND,?Priority:=MAX_PRIORITY)
            else if(PlayerBakery.PlayerDataCache.Language=FRENCH):
                GameManager.DisplayCharacterMessage(Agent,I_CHARACTER,"Les ingrédients inutilisés sont aux dépôts.",?DisplayTime:=DelayTime,?Sound:=MESSAGE_SOUND,?Priority:=MAX_PRIORITY)
            else if(PlayerBakery.PlayerDataCache.Language=PORTUGUESE):
                GameManager.DisplayCharacterMessage(Agent,I_CHARACTER,"Os ingredientes não usados estão nos depósitos.",?DisplayTime:=DelayTime,?Sound:=MESSAGE_SOUND,?Priority:=MAX_PRIORITY)            
            Sleep(DelayTime+0.25)
            #CAMERA 8
            Camera8.AddTo(Agent)
            set DelayTime = 5.0
            GameManager.CleanCharacterMessages(Agent)
            GameManager.CleanCharacterMessages(Agent)
            if(PlayerBakery.PlayerDataCache.Language=ENGLISH):
                GameManager.DisplayCharacterMessage(Agent,I_CHARACTER,"You can use coal in the Main Generator to boost production.",?DisplayTime:=DelayTime,?Sound:=MESSAGE_SOUND,?Priority:=MAX_PRIORITY)
            else if(PlayerBakery.PlayerDataCache.Language=SPANISH):
                GameManager.DisplayCharacterMessage(Agent,I_CHARACTER,"Usa carbón en el Generador para aumentar producción.",?DisplayTime:=DelayTime,?Sound:=MESSAGE_SOUND,?Priority:=MAX_PRIORITY)
            else if(PlayerBakery.PlayerDataCache.Language=FRENCH):
                GameManager.DisplayCharacterMessage(Agent,I_CHARACTER,"Utilisez du charbon pour booster la production.",?DisplayTime:=DelayTime,?Sound:=MESSAGE_SOUND,?Priority:=MAX_PRIORITY)
            else if(PlayerBakery.PlayerDataCache.Language=PORTUGUESE):
                GameManager.DisplayCharacterMessage(Agent,I_CHARACTER,"Use carvão no Gerador para aumentar produção.",?DisplayTime:=DelayTime,?Sound:=MESSAGE_SOUND,?Priority:=MAX_PRIORITY)
                    
            Sleep(DelayTime+0.25)
            #CAMERA 9
            Camera9.AddTo(Agent)
            set DelayTime = 5.0
            GameManager.CleanCharacterMessages(Agent)
            if(PlayerBakery.PlayerDataCache.Language=ENGLISH):
                GameManager.DisplayCharacterMessage(Agent,I_CHARACTER,"This generates Super Energy. You can check its status here.",?DisplayTime:=DelayTime,?Sound:=MESSAGE_SOUND,?Priority:=MAX_PRIORITY)
            else if(PlayerBakery.PlayerDataCache.Language=SPANISH):
                GameManager.DisplayCharacterMessage(Agent,I_CHARACTER,"Esto genera Super Energía. Revisa su estado aquí.",?DisplayTime:=DelayTime,?Sound:=MESSAGE_SOUND,?Priority:=MAX_PRIORITY)
            else if(PlayerBakery.PlayerDataCache.Language=FRENCH):
                GameManager.DisplayCharacterMessage(Agent,I_CHARACTER,"Cela génère de l'Énergie Super. Vérifiez son état ici.",?DisplayTime:=DelayTime,?Sound:=MESSAGE_SOUND,?Priority:=MAX_PRIORITY)
            else if(PlayerBakery.PlayerDataCache.Language=PORTUGUESE):
                GameManager.DisplayCharacterMessage(Agent,I_CHARACTER,"Isso gera Super Energia. Verifique seu status aqui.",?DisplayTime:=DelayTime,?Sound:=MESSAGE_SOUND,?Priority:=MAX_PRIORITY)            
            Sleep(DelayTime+0.25)
            GameManager.CleanCharacterMessages(Agent)
            if(PlayerBakery.PlayerDataCache.Language=ENGLISH):
                GameManager.DisplayCharacterMessage(Agent,G_CHARACTER,"Good luck!",?DisplayTime:=5.0,?Priority:=MAX_PRIORITY)
            else if(PlayerBakery.PlayerDataCache.Language=SPANISH):
                GameManager.DisplayCharacterMessage(Agent,G_CHARACTER,"¡Buena suerte!",?DisplayTime:=5.0,?Priority:=MAX_PRIORITY)
            else if(PlayerBakery.PlayerDataCache.Language=FRENCH):
                GameManager.DisplayCharacterMessage(Agent,G_CHARACTER,"Bonne chance!",?DisplayTime:=5.0,?Priority:=MAX_PRIORITY)
            else if(PlayerBakery.PlayerDataCache.Language=PORTUGUESE):
                GameManager.DisplayCharacterMessage(Agent,G_CHARACTER,"Boa sorte!",?DisplayTime:=5.0,?Priority:=MAX_PRIORITY)
            PlayerBakery.CheckQuestProgress(1.0,BAKERTUTORIAL_TUTORIAL_TYPE)
            MainInventoryUI.SimplifiedRefreshAll()
            FortCharacter.ReleaseFromStasis()
            Camera0.RemoveFrom(Agent)
            Camera1.RemoveFrom(Agent)
            Camera2.RemoveFrom(Agent)
            Camera3.RemoveFrom(Agent)
            Camera4.RemoveFrom(Agent)
            Camera5.RemoveFrom(Agent)
            Camera6.RemoveFrom(Agent)
            Camera7.RemoveFrom(Agent)
            Camera8.RemoveFrom(Agent)
            Camera9.RemoveFrom(Agent)
            for(Index:=0..AllAgentsWatchingTutorial.Length-1, AgentTutorial := AllAgentsWatchingTutorial[Index], Agent = AgentTutorial, AuxArray := AllAgentsWatchingTutorial.RemoveElement[Index]):
                set AllAgentsWatchingTutorial = AuxArray
        else:
            PrintFatalError("On PlayBakerTutorial, One camera was invalid")


    QuestCompletedEvent<private>(PlayerData : player_data_table, Player : player) : player_data_table =
        NumQuestsCompleted := PlayerData.NumQuestsCompleted + 1
        var UpdatedTable : player_data_table = PlayersDataManager.SetNumQuestsCompleted(PlayerData,NumQuestsCompleted)
        Log("Now player has quest {NumQuestsCompleted + 1}",LOG_LEVEL_DEBUG5)
        if(NumQuestsCompleted > Quest0_Index): #Create delay if the quest is now bellow first real quest 
            set UpdatedTable = PlayersDataManager.CreateQuest_Delay(UpdatedTable, Quest_DelayTime_Base)
            ##Print("Quest delay created")
            if(GameManager := MaybeGameManager?, Agent := agent[Player]):  
                GameManager.RemoveQuest_MapTracker(Agent)
        else:
            CheckLore(Player,PlayerData)
            CheckBakerUnlocks(Player,PlayerData.NumQuestsCompleted)
            Quest:=GetNewQuest(NumQuestsCompleted,PlayerData) #Grant Quest inmidiatly if not
            if(GameManager := MaybeGameManager?, Agent := agent[Player], AgentBakery := GameManager.GetAgentBakery_Transacted(Agent)?):
                AgentBakery.CloseAllTutorialAffectedUI()
                GameManager.AssignQuest_MapTracker(Agent,NumQuestsCompleted,Quest(3))
            Log("Quest Objetive Type: {Quest(3)},  Award Type : {Quest(5)}, Award Amount: {Quest(6)}",LOG_LEVEL_DEBUG5)
            set UpdatedTable = PlayersDataManager.CreateNewQuest(UpdatedTable,NumQuestsCompleted,Quest(3),Quest(5),Quest(6))
        return UpdatedTable

    

    CheckBakerUnlocks(Player : player, NumQuestCompleted : int) : void =
        if(Agent := agent[Player], GameManager := MaybeGameManager?):
            MaybeBaker:=GameManager.GetAgentBakery(Agent)
            if(ValidBaker := MaybeBaker?):
                if(NumQuestCompleted>=16):
                    ValidBaker.UnlockCBB()
                if(NumQuestCompleted>=22):
                    ValidBaker.UnlockPCSB()

    OrderCompletedEvent<private>(PlayerData : player_data_table, Player : player) : player_data_table =
        var UpdatedTable : player_data_table = PlayerData
        if(UpdatedTable.NumQuestsCompleted > TutorialOrder_Index): #Create delay if the quest is now bellow first real quest 
            #TupleOrder:=GetOrder()
            set UpdatedTable = PlayersDataManager.CreateOrder_Delay(UpdatedTable, Order_DelayTime_Base)
        return UpdatedTable


    GetQuestObjetiveTransform(NumQuestsCompleted : int, QuestType :int, Agent : agent, PlayerData : player_data_table, NumQuest : int) : tuple(logic,transform) =
        var OutPutTransform : transform = transform{} 
        var ValidTransform : logic = false
        if(NumQuestsCompleted>Quest0_Index, NormalQuestTupleData := AllNormalQuests[NumQuest]):
            if(ValidProp := NormalQuestTupleData(7)?):
                set OutPutTransform = ValidProp.GetTransform()
                set ValidTransform = true
            else:
                #if(CHECKINFO_TUTORIAL_TYPE=NormalQuestTupleData(3)):
                #    Log("CHECKINFO_TUTORIAL_TYPE Detected!", LOG_LEVEL_DEBUG7)
                MaybeButton:=GetButtonDevice(NumQuestsCompleted, NumQuest,QuestType,Agent)
                if(ValidButton := MaybeButton?):
                    set OutPutTransform = ValidButton.GetTransform()
                    set ValidTransform = true
        else if(TutorialQuestTupleData := AllTutorialQuests[NumQuest]):
            if(ValidProp := TutorialQuestTupleData(7)?):
                set OutPutTransform = ValidProp.GetTransform()
                set ValidTransform = true
            else:
                MaybeButton:=GetButtonDevice(NumQuestsCompleted, NumQuest,QuestType,Agent)
                if(ValidButton := MaybeButton?):
                    set OutPutTransform = ValidButton.GetTransform()
                    set ValidTransform = true
        else:
            if(NumQuestsCompleted<>-99, PlayerData.NumQuestsCompleted<Quest0_Index):
                PrintFatalError("On GetQuestObjetiveProp. Quest {NumQuest} or QuestType {QuestType} not found!")
        return (ValidTransform,OutPutTransform)



    GetNewQuest(NumQuestCompleted : int, PlayerDataCache : player_data_table) : tuple(string,texture,string,int,float,int,float,?creative_prop,?button_device,int) =
        CheckQuestInited()
        #Log("NumQuestCompleted: {NumQuestCompleted}", LOG_LEVEL_DEBUG6)
        if(NumQuestCompleted>Quest0_Index):
            #CheckIfMasteryWilGrantSapphires
                return GeNormalQuest(PlayerDataCache)
        else:
            return GetTutorialQuest(NumQuestCompleted)

    GetCurrentQuest(NumQuestsCompleted : int, PlayerDataCache : player_data_table) : tuple(string,texture,string,int,float,int,float,?creative_prop,?button_device,int) =
        CheckQuestInited()

        var QuestInfo : tuple(string,texture,string,int,float,int,float,?creative_prop,?button_device,int) = ("Teleport Testing",ParkourIcon,"Spend Coal on teleporters",USEEXPENSIVETELEPORTER_TUTORIAL_TYPE,10.0,COAL,StandarQuestReward_Coal,false,false,-1)
        #Log("NumQuestCompleted: {NumQuestCompleted}", LOG_LEVEL_DEBUG6)
        if(NumQuestsCompleted>Quest0_Index, CurrentQuest := AllNormalQuests[PlayerDataCache.QuestAssigned_NumQuest]):
            set QuestInfo = (CurrentQuest(0),CurrentQuest(1),CurrentQuest(2),CurrentQuest(3),CurrentQuest(4),CurrentQuest(5),CurrentQuest(6),CurrentQuest(7),CurrentQuest(8),PlayerDataCache.QuestAssigned_NumQuest)
        else:
            return GetTutorialQuest(NumQuestsCompleted)
        return QuestInfo

    SetGameManager(GameManager : game_manager) : void =
        set MaybeGameManager = option{GameManager}

    GeNormalQuest<private>(PlayerDataCache : player_data_table) : tuple(string,texture,string,int,float,int,float,?creative_prop,?button_device,int) =
        Log("Getting normal Quest",LOG_LEVEL_DEBUG10)
        var QuestInfo : tuple(string,texture,string,int,float,int,float,?creative_prop,?button_device,int) = ("Teleport Testing",ParkourIcon,"Spend Coal on teleporters",USEEXPENSIVETELEPORTER_TUTORIAL_TYPE,10.0,COAL,StandarQuestReward_Coal,false,false,-1)
        
        var AllValidQuests  : []tuple(string,texture,string,int,float,int,float,?creative_prop,?button_device) = array{}
        var AllValidQuestIndexes : []int = array{}

        var AgedSapphiresAward : float = 0.0
        PickaxeTotalStrength:=CalculeTotalStrength(PlayerDataCache.TLV_PickaxeStrength)
        if(GameManager:=MaybeGameManager?):
            set AgedSapphiresAward=CalculeAgedSapphires(PlayerDataCache.LV_Area, PickaxeTotalStrength, GameManager)
        if(not PlayerDataCache.IsMasteryQuestCompleted?, AgedSapphiresAward>0.0):
            Log("Is going to be granted a MasterySPECIAL Quest",LOG_LEVEL_DEBUG10)
            if(SelectedQuest := AllNormalQuests[MASTERYSPECIALQUEST_INDEX], NumQuest:=MASTERYSPECIALQUEST_INDEX):
                Log("Granted a MasterySPECIAL Quest",LOG_LEVEL_DEBUG10)
                set QuestInfo = (SelectedQuest(0),SelectedQuest(1),SelectedQuest(2),SelectedQuest(3),SelectedQuest(4),SelectedQuest(5),SelectedQuest(6),SelectedQuest(7),SelectedQuest(8),NumQuest)
        else:
            if(PlayerDataCache.LV_Area>=17):
                set AllValidQuests = AllNormalQuests
                for(Index:=0..MASTERYSPECIALQUEST_INDEX-1): #MAX TO QUEST 13, BECAUSE 14 IS THE MASTERY SPECIAL QUEST
                    set AllValidQuestIndexes = AllValidQuestIndexes + array{Index}
            else if(PlayerDataCache.LV_Area>=7):
                for(Index:=0..8, Quest := AllNormalQuests[Index]):
                    set AllValidQuestIndexes = AllValidQuestIndexes + array{Index}
                    set AllValidQuests = AllValidQuests + array{Quest}
                for(Index:=9..11, Quest := AllNormalQuests[Index]):
                    set AllValidQuestIndexes = AllValidQuestIndexes + array{Index}
                    set AllValidQuests = AllValidQuests + array{Quest}
            else if(PlayerDataCache.LV_Area>=1):
                for(Index:=0..7, Quest := AllNormalQuests[Index]):
                    set AllValidQuestIndexes = AllValidQuestIndexes + array{Index}
                    set AllValidQuests = AllValidQuests + array{Quest}
                for(Index:=9..11, Quest := AllNormalQuests[Index]):
                    set AllValidQuestIndexes = AllValidQuestIndexes + array{Index}
                    set AllValidQuests = AllValidQuests + array{Quest}
            else:
                for(Index:=0..6, Quest := AllNormalQuests[Index]):
                    set AllValidQuestIndexes = AllValidQuestIndexes + array{Index}
                    set AllValidQuests = AllValidQuests + array{Quest}
                for(Index:=9..11, Quest := AllNormalQuests[Index]):
                    set AllValidQuestIndexes = AllValidQuestIndexes + array{Index}
                    set AllValidQuests = AllValidQuests + array{Quest}

            Log("AllValidQuestIndexes.Length: {AllValidQuestIndexes.Length}",LOG_LEVEL_DEBUG5)
            Log("AllValidQuests.Length: {AllValidQuests.Length}",LOG_LEVEL_DEBUG5)

            RandomQuest:=GetRandomInt(0,AllValidQuests.Length-1)
            if(SelectedQuest := AllValidQuests[RandomQuest], NumQuest:=AllValidQuestIndexes[RandomQuest]):
                Log("Getting normal Quest: {NumQuest}",LOG_LEVEL_DEBUG5)
                set QuestInfo = (SelectedQuest(0),SelectedQuest(1),SelectedQuest(2),SelectedQuest(3),SelectedQuest(4),SelectedQuest(5),SelectedQuest(6),SelectedQuest(7),SelectedQuest(8),NumQuest)
            else:
                PrintFatalError("Selected Quest RandomQuest: {RandomQuest} not found!")
        return QuestInfo

    GetTutorialQuest<private>(NumQuest : int) : tuple(string,texture,string,int,float,int,float,?creative_prop,?button_device,int) =
        var QuestInfo : tuple(string,texture,string,int,float,int,float,?creative_prop,?button_device,int) = ("Teleport Testing",ParkourIcon,"Spend Coal on teleporters",USEEXPENSIVETELEPORTER_TUTORIAL_TYPE,10.0,COAL,StandarQuestReward_Coal,false,false,-1)
        if(var Quest : tuple([]char,texture,[]char,int,float,int,float,?creative_prop,?button_device) = AllTutorialQuests[NumQuest]): 
           # #Print("Granting quest {NumQuest}")
           set QuestInfo = (Quest(0),Quest(1),Quest(2),Quest(3),Quest(4),Quest(5),Quest(6),Quest(7),Quest(8),-1)
        else:
            PrintFatalError("On GetTutorialQuest")
        return QuestInfo

    GetOrder(BakeryManager : bakery_machines_manager) : tuple(int,float,int,float) =

        #Calcule Pickaxe Resourcers per pick
        TuplePickDamage := BakeryManager.GetBakeryPickaxeDamage(BakeryManager.PlayerDataCache.LV_Area+1,false,true)
        var TotalDamage : float = TuplePickDamage(0)

        var ResourcersAmmount_PerPick : float = BakeryManager.ApplyEfficiencyToDamage(TotalDamage)

        Log("Calculating Order",LOG_LEVEL_DEBUG10)
        Log("ResourcersAmmount_PerPick: {GetOptimizedMagnitudeLetter(ResourcersAmmount_PerPick)} without Resource Efficiency",LOG_LEVEL_DEBUG10)

        AllEfficiency := BakeryManager.AllPickaxeEfficiency
        var AVGEfficiency : float = 0.0

        for(Res:=WATER..PIMENTOS, ResEfficency := AllEfficiency[Res]):
            set AVGEfficiency = AVGEfficiency + ResEfficency

        set AVGEfficiency = AVGEfficiency / (4*1.0)

        Log("AVGEfficiency: {GetOptimizedMagnitudeLetter(AVGEfficiency)}",LOG_LEVEL_DEBUG10)

        set ResourcersAmmount_PerPick = ResourcersAmmount_PerPick*AVGEfficiency
        Log("ResourcersAmmount_PerPick: {GetOptimizedMagnitudeLetter(ResourcersAmmount_PerPick)}",LOG_LEVEL_DEBUG10)
        
        #Select the two machines with less production
        BProduction := BakeryManager.BB_Production
        CBProduction := BakeryManager.CBB_Production
        PCSProduction := BakeryManager.PCSB_Production
        
        #Selecting the Baker
        var AllProductsProduction : []float = array{BProduction,CBProduction,PCSProduction}
        var IndexBiggestProduction : int = 0
        var BiggestProduction : float = 0.0
        for(Index:=0..AllProductsProduction.Length-1, Production:=AllProductsProduction[Index], Production > BiggestProduction):
            set BiggestProduction = Production
            set IndexBiggestProduction = Index
        if(AuxArray := AllProductsProduction.RemoveElement[IndexBiggestProduction], set AllProductsProduction = AuxArray):
        RandomBakerIndex := GetRandomInt(0,1)

        #Calculate Total Resourcers Required
        NumPicksPerOrder := GetRandomFloat(NumMaxPicks_PerOrder/2.0,NumMaxPicks_PerOrder)
        TotalResourcersRequired := ResourcersAmmount_PerPick * NumPicksPerOrder
        Log("NumPicksPerOrder: {NumPicksPerOrder} TotalResourcersRequired: {GetOptimizedMagnitudeLetter(TotalResourcersRequired)}",LOG_LEVEL_DEBUG10)
        var TotalResourcersPerProduct : float = 100.0
        var ProductRequested : int = B_TYPE
        if(RandomBakerProduction := AllProductsProduction[RandomBakerIndex]):
            if(RandomBakerProduction=BProduction):
                set TotalResourcersPerProduct = BakeryManager.Req_WaterFor_B + BakeryManager.Req_FlourFor_B
                set ProductRequested = B_TYPE
            else if(RandomBakerProduction=CBProduction):
                set TotalResourcersPerProduct = BakeryManager.Req_WaterFor_CB + BakeryManager.Req_FlourFor_CB + BakeryManager.Req_CheeseFor_CB
                set ProductRequested = CB_TYPE
            else if(RandomBakerProduction=PCSProduction):
                set TotalResourcersPerProduct = BakeryManager.Req_WaterFor_PCS + BakeryManager.Req_FlourFor_PCS + BakeryManager.Req_CheeseFor_PCS + BakeryManager.Req_PimentosFor_PCS
                set ProductRequested = PCS_TYPE 
        var NumProductsRequired : float = TotalResourcersRequired / TotalResourcersPerProduct
        Log("NumProductsRequired: {GetOptimizedMagnitudeLetter(TotalResourcersRequired)} / {GetOptimizedMagnitudeLetter(TotalResourcersPerProduct)} = {GetOptimizedMagnitudeLetter(NumProductsRequired)}",LOG_LEVEL_DEBUG10)
        if(NumProductsRequired<10.0):
            set NumProductsRequired = 10.0
        CoalAmount:=GetRandomInt(MaxCoal_PerOrder-10,MaxCoal_PerOrder) 
        return(ProductRequested,NumProductsRequired,COAL,CoalAmount*1.0)

    CheckIfTutorialWasCompletedBefore(Player : player, PlayerData : player_data_table) : player_data_table = 
        var PlayerDataTable : player_data_table = PlayerData
        #QuestAssigned_NumQuest
        #Log("NumQuestCompleted: {PlayerData.NumQuestCompleted}, LV_Area: {PlayerData.LV_Area}, QuestAssigned_NumQuest: {PlayerData.QuestAssigned_NumQuest}",LOG_LEVEL_DEBUG4)
        if(PlayerDataTable.QuestAssigned_NumQuest=4):
            set PlayerDataTable = AddAgentQuest_ProgressV2(Player,PlayerDataTable,1.0,BAKERTUTORIAL_TUTORIAL_TYPE)(4)
        else if(PlayerDataTable.QuestAssigned_NumQuest=6, PlayerDataTable.LV_Area>=0):
            Log("Quest 6 was completed before, trying to complete...",LOG_LEVEL_DEBUG5)
            set PlayerDataTable = AddAgentQuest_ProgressV2(Player,PlayerDataTable,1.0,UNLOCKAREA_TUTORIAL_TYPE)(4)
        else if(PlayerDataTable.QuestAssigned_NumQuest=11):
            if(LVStrength := PlayerDataTable.TLV_PickaxeStrength[0]):
                if(LVStrength>0):
                    Log("Quest 11 was completed before, trying to complete...",LOG_LEVEL_DEBUG5)
                    set PlayerDataTable = AddAgentQuest_ProgressV2(Player,PlayerDataTable,1.0,IMPROVEPICKAXE_TUTORIAL_TYPE)(4)
        else if(PlayerDataTable.QuestAssigned_NumQuest=12):    
            var MustBeCompleted : logic = false
            for(Limit : PlayerDataTable.AllLV_BackpackLimits, Limit > 0):
                set MustBeCompleted = true
            if(MustBeCompleted?):
                Log("Quest 12 was completed before, trying to complete...",LOG_LEVEL_DEBUG5)
                set PlayerDataTable = AddAgentQuest_ProgressV2(Player,PlayerDataTable,1.0,IMPROVEBACKPACK_TUTORIAL_TYPE)(4)
        else if(PlayerDataTable.QuestAssigned_NumQuest=13): 
            var MustBeCompleted : logic = false
            for(Deposit : PlayerDataTable.AllLV_MainDeposits, Deposit > 0):
                set MustBeCompleted = true
            if(MustBeCompleted?):
                Log("Quest 13 was completed before, trying to complete...",LOG_LEVEL_DEBUG5)
                set PlayerDataTable = AddAgentQuest_ProgressV2(Player,PlayerDataTable,1.0,IMPROVEMAINDEPOSITS_TUTORIAL_TYPE)(4)
        else if(PlayerDataTable.QuestAssigned_NumQuest=14): 
            var MustBeCompleted : logic = false
            if(LVStrength := PlayerDataTable.TLV_BB[0]):
                if(LVStrength>0):
                    Log("Quest 14 was completed before, trying to complete...",LOG_LEVEL_DEBUG5)
                    set PlayerDataTable = AddAgentQuest_ProgressV2(Player,PlayerDataTable,1.0,IMPROVEBAKERY_TUTORIAL_TYPE)(4)
        else if(PlayerDataTable.QuestAssigned_NumQuest=16): 
            var MustBeCompleted : logic = false
            for(Index:=0..PlayerDataTable.AllLV_PickaxePerks.Length-1,PickaxePerk := PlayerDataTable.AllLV_PickaxePerks[Index]):
                if(PickaxePerk>1):
                    set MustBeCompleted = true
                else if(Index<4, PickaxePerk > 0):
                    set MustBeCompleted = true
            if(MustBeCompleted?):
                Log("Quest 16 was completed before, trying to complete...",LOG_LEVEL_DEBUG5)
                set PlayerDataTable = AddAgentQuest_ProgressV2(Player,PlayerDataTable,1.0,IMPROVEPICKPERK_TUTORIAL_TYPE)(4)
        else if(PlayerDataTable.QuestAssigned_NumQuest=25, PlayerDataTable.LV_Area>=1):
            Log("Quest 25 was completed before, trying to complete...",LOG_LEVEL_DEBUG5)
            set PlayerDataTable = AddAgentQuest_ProgressV2(Player,PlayerDataTable,1.0,UNLOCKAREA_TUTORIAL_TYPE)(4)
        else if(PlayerDataTable.QuestAssigned_NumQuest=27, PlayerDataTable.AllPets.Length>0):
            Log("Quest 27 was completed before, trying to complete...",LOG_LEVEL_DEBUG5)
            set PlayerDataTable = AddAgentQuest_ProgressV2(Player,PlayerDataTable,1.0,CREATEPET_TUTORIAL_TYPE)(4)
        else if(PlayerDataTable.QuestAssigned_NumQuest=28):
            var MustBeCompleted : logic = false
            for(Pet : PlayerDataTable.AllPets, Pet.LVPet_PackageA>1 or Pet.LVPet_PackageB>1):
                set MustBeCompleted = true
            if(MustBeCompleted?):
                Log("Quest 28 was completed before, trying to complete...",LOG_LEVEL_DEBUG5)
                set PlayerDataTable = AddAgentQuest_ProgressV2(Player,PlayerDataTable,1.0,IMPROVEPET_TUTORIAL_TYPE)(4)
        else if(PlayerDataTable.QuestAssigned_NumQuest=29):
            var IsAlreadyCompleted : logic = false
            for(Indexes:=0..PlayerDataTable.AllHexaGame_TotemsUnlocked.Length-1, IsCompleted:=PlayerDataTable.AllHexaGame_TotemsUnlocked[Indexes], IsCompleted?):
                set IsAlreadyCompleted = true
            if(IsAlreadyCompleted?):
                set PlayerDataTable = AddAgentQuest_ProgressV2(Player,PlayerDataTable,1.0,UNLOCKHEXAGAME_TUTORIAL_TYPE)(4)
                Log("Quest 29 was completed before, trying to complete...",LOG_LEVEL_DEBUG9)
        else if(PlayerDataTable.QuestAssigned_NumQuest=34):
            if(PlayerDataTable.Tier_Sword>1 or PlayerDataTable.LV_SwordDamage>1 or PlayerDataTable.LV_SwordCriticalProb>1 or PlayerDataTable.LV_SwordCriticalDamage>1 or PlayerDataTable.LV_SwordReward>1 or PlayerDataTable.Tier_Bow>1 or PlayerDataTable.LV_BowDamage>1 or PlayerDataTable.LV_BowCriticalProb>1 or PlayerDataTable.LV_BowCriticalDamage>1 or PlayerDataTable.LV_BowReward>1):
                set PlayerDataTable = AddAgentQuest_ProgressV2(Player,PlayerDataTable,1.0,WEAPONIMPROVE_TUTORIAL_TYPE)(4)
                Log("Quest 34 was completed before, trying to complete...",LOG_LEVEL_DEBUG5)
        else if(PlayerDataTable.QuestAssigned_NumQuest=35, PlayerDataTable.Dungeon_LVDelay > 1 or PlayerDataTable.HexaGame_LVDelay > 1):
            set PlayerDataTable = AddAgentQuest_ProgressV2(Player,PlayerDataTable,1.0,TOKENIMPROVE_TUTORIAL_TYPE)(4)
            Log("Quest 35 was completed before, trying to complete...",LOG_LEVEL_DEBUG5)
        else if(PlayerDataTable.QuestAssigned_NumQuest=36, PlayerDataTable.Dungeon_LVAccumulated > 1 or PlayerDataTable.HexaGame_LVAccumulated > 1):
            set PlayerDataTable = AddAgentQuest_ProgressV2(Player,PlayerDataTable,1.0,AGGEDSAPPHIRESUSED_TUTORIAL_TYPE)(4)
            Log("Quest 36 was completed before, trying to complete...",LOG_LEVEL_DEBUG5)
        return PlayerDataTable
        
    CheckLore(Player : player, PlayerData : player_data_table, ?SpecialEvent : int = -1) : void =
        if(GameManager := MaybeGameManager?, Agent := agent[Player], Player.IsActive[]): 
            if(SpecialEvent=-1): 
                if(PlayerData.NumQuestsCompleted=7):
                    if(PlayerData.Language=ENGLISH):
                        GameManager.DisplayCharacterMessage(Agent,G_CHARACTER,"Each slate bread type grants a unique resource!")
                    if(PlayerData.Language=SPANISH):
                        GameManager.DisplayCharacterMessage(Agent,G_CHARACTER,"Cada pan de pizarra da un recurso único!")
                    if(PlayerData.Language=FRENCH):
                        GameManager.DisplayCharacterMessage(Agent,G_CHARACTER,"Chaque pain ardoise donne une ressource unique!")
                    if(PlayerData.Language=PORTUGUESE):
                        GameManager.DisplayCharacterMessage(Agent,G_CHARACTER,"Cada pão de ardósia dá um recurso único!")
                
                if(PlayerData.NumQuestsCompleted=15):
                    if(PlayerData.Language=ENGLISH):
                        GameManager.DisplayCharacterMessage(Agent,G_CHARACTER,"Improving Baker production will increase the resources needed!")
                    if(PlayerData.Language=SPANISH):
                        GameManager.DisplayCharacterMessage(Agent,G_CHARACTER,"Mejorar la producción aumenta el consumo!")
                    if(PlayerData.Language=FRENCH):
                        GameManager.DisplayCharacterMessage(Agent,G_CHARACTER,"Améliorer la production consomme plus!")
                    if(PlayerData.Language=PORTUGUESE):
                        GameManager.DisplayCharacterMessage(Agent,G_CHARACTER,"Melhorar a produção aumenta o consumo!")
                
                if(PlayerData.NumQuestsCompleted=16):
                    if(PlayerData.Language=ENGLISH):
                        GameManager.DisplayCharacterMessage(Agent,G_CHARACTER,"You can now use the Cheese Bread Baker!")
                    if(PlayerData.Language=SPANISH):
                        GameManager.DisplayCharacterMessage(Agent,G_CHARACTER,"¡Ahora puedes usar el Pan de Queso!")
                    if(PlayerData.Language=FRENCH):
                        GameManager.DisplayCharacterMessage(Agent,G_CHARACTER,"Vous pouvez utiliser le Pain au Fromage!")
                    if(PlayerData.Language=PORTUGUESE):
                        GameManager.DisplayCharacterMessage(Agent,G_CHARACTER,"Agora você pode usar o Pão de Queijo!")
                
                if(PlayerData.NumQuestsCompleted=17):
                    if(PlayerData.Language=ENGLISH):
                        GameManager.DisplayCharacterMessage(Agent,G_CHARACTER,"Mutant Material drops when you break repeatedly-hit Slate Bread!")
                    if(PlayerData.Language=SPANISH):
                        GameManager.DisplayCharacterMessage(Agent,G_CHARACTER,"El Material Mutante cae al romper pan muy golpeado!")
                    if(PlayerData.Language=FRENCH):
                        GameManager.DisplayCharacterMessage(Agent,G_CHARACTER,"Le Matériau Mutant tombe en cassant un pain frappé!")
                    if(PlayerData.Language=PORTUGUESE):
                        GameManager.DisplayCharacterMessage(Agent,G_CHARACTER,"O Material Mutante cai ao quebrar pão muito golpeado!")
                
                if(PlayerData.NumQuestsCompleted=18):
                    if(PlayerData.Language=ENGLISH):
                        GameManager.DisplayCharacterMessage(Agent,G_CHARACTER,"A green valve has more priority than a yellow one")
                    if(PlayerData.Language=SPANISH):
                        GameManager.DisplayCharacterMessage(Agent,G_CHARACTER,"La válvula verde tiene más prioridad que la amarilla!")
                    if(PlayerData.Language=FRENCH):
                        GameManager.DisplayCharacterMessage(Agent,G_CHARACTER,"La vanne verte est prioritaire sur la jaune!")
                    if(PlayerData.Language=PORTUGUESE):
                        GameManager.DisplayCharacterMessage(Agent,G_CHARACTER,"A válvula verde tem prioridade sobre a amarela!")
                
                if(PlayerData.NumQuestsCompleted=19):
                    if(PlayerData.Language=ENGLISH):
                        GameManager.DisplayCharacterMessage(Agent,G_CHARACTER,"A red valve is fully closed. Don’t forget this!")
                    if(PlayerData.Language=SPANISH):
                        GameManager.DisplayCharacterMessage(Agent,G_CHARACTER,"La válvula roja está cerrada por completo!")
                    if(PlayerData.Language=FRENCH):
                        GameManager.DisplayCharacterMessage(Agent,G_CHARACTER,"La vanne rouge est complètement fermée!")
                    if(PlayerData.Language=PORTUGUESE):
                        GameManager.DisplayCharacterMessage(Agent,G_CHARACTER,"A válvula vermelha está totalmente fechada!")
                
                if(PlayerData.NumQuestsCompleted=25):
                    if(PlayerData.Language=ENGLISH):
                        GameManager.DisplayCharacterMessage(Agent,G_CHARACTER,"Coal Generator boosts x{GetMagnitudeLetter(SuperEnergyMultiplier_Owned,"","")} Efficiency and Baker Speed!")
                    if(PlayerData.Language=SPANISH):
                        GameManager.DisplayCharacterMessage(Agent,G_CHARACTER,"El Generador da x{GetMagnitudeLetter(SuperEnergyMultiplier_Owned,"","")} eficiencia y velocidad!")
                    if(PlayerData.Language=FRENCH):
                        GameManager.DisplayCharacterMessage(Agent,G_CHARACTER,"Le Générateur donne x{GetMagnitudeLetter(SuperEnergyMultiplier_Owned,"","")} d'efficacité et vitesse!")
                    if(PlayerData.Language=PORTUGUESE):
                        GameManager.DisplayCharacterMessage(Agent,G_CHARACTER,"O Gerador dá x{GetMagnitudeLetter(SuperEnergyMultiplier_Owned,"","")} eficiência e velocidade!")
                
                if(PlayerData.NumQuestsCompleted=27):
                    if(PlayerData.Language=ENGLISH):
                        GameManager.DisplayCharacterMessage(Agent,G_CHARACTER,"The Mutant Material allows for a purer Sourdough!")
                    if(PlayerData.Language=SPANISH):
                        GameManager.DisplayCharacterMessage(Agent,G_CHARACTER,"El Material Mutante hace un Sourdough más puro!")
                    if(PlayerData.Language=FRENCH):
                        GameManager.DisplayCharacterMessage(Agent,G_CHARACTER,"Le Matériau Mutant rend le Sourdough plus pur!")
                    if(PlayerData.Language=PORTUGUESE):
                        GameManager.DisplayCharacterMessage(Agent,G_CHARACTER,"O Material Mutante faz um Sourdough mais puro!")
                
                if(PlayerData.NumQuestsCompleted=36):
                    if(PlayerData.Language=ENGLISH):
                        GameManager.DisplayCharacterMessage(Agent,G_CHARACTER,"You can now complete extra Quests!")
                    if(PlayerData.Language=SPANISH):
                        GameManager.DisplayCharacterMessage(Agent,G_CHARACTER,"¡Ahora puedes completar misiones extra!")
                    if(PlayerData.Language=FRENCH):
                        GameManager.DisplayCharacterMessage(Agent,G_CHARACTER,"Vous pouvez maintenant faire des quêtes en plus!")
                    if(PlayerData.Language=PORTUGUESE):
                        GameManager.DisplayCharacterMessage(Agent,G_CHARACTER,"Agora você pode completar missões extras!")
                
                        
            else if(SpecialEvent = SPECIALEVENT_MAINLORE):
                #if(LoreState:=PlayerData.AllLoreStates[MAIN_LORE], LoreState-2 <= PlayerData.LV_Area):
                    spawn:
                        RunLore(Player)
            else:
                if(SpecialEvent = SPECIALEVENT_FIRSTENCOUNTER, FirstEncounterString := AllWLore[2]):
                    GameManager.DisplayCharacterMessage(Agent,W_CHARACTER,FirstEncounterString) 

    DelayCheckWelcomeLore(Player : player, AgentBakery : bakery_machines_manager, Delay : float)<suspends> : void =
        if(LoreState := AgentBakery.PlayerDataCache.AllLoreStates[MAIN_LORE], LoreState<2):
            Sleep(Delay)
            Log("LoreState: {LoreState}", LOG_LEVEL_DEBUG7)
            RunLore(Player, ?PreDelay := 0.0)
            
        if(LoreState := AgentBakery.PlayerDataCache.AllLoreStates[MAIN_LORE], LoreState<2):
            Sleep(1.0)
            Log("LoreState: {LoreState}", LOG_LEVEL_DEBUG7)
            RunLore(Player, ?PreDelay := 0.0)
            #AgentBakery.ForceLanguageSelection()
            

    RunLore<private>(Player : player, ?PreDelay : float = 1.0)<suspends> : void =
        if(GameManager := MaybeGameManager?):
            if(Player.IsActive[], Agent := agent[Player], AgentBakery := GameManager.GetAgentBakery_Transacted(Agent)?, Character := Agent.GetFortCharacter[]):
                if(LoreState := AgentBakery.PlayerDataCache.AllLoreStates[MAIN_LORE]):
                    Log("LoreState: {LoreState} | Player LV_Area: {AgentBakery.PlayerDataCache.LV_Area}", LOG_LEVEL_DEBUG7)
                    if(LoreState<2):
                        Log("Executing Welcome Lore!", LOG_LEVEL_DEBUG7)
                        if(Player.IsActive[]):
                            NewLoreState := LoreState + 1
                            AgentBakery.SetNewPlayerDataCache(GameManager.PlayersDataManager.SetLoreState(AgentBakery.PlayerDataCache,MAIN_LORE,NewLoreState))
                        Sleep(PreDelay)
                        var MaybeCamera : ?gameplay_camera_fixed_point_device = false
                        if(LoreCamera := AllLoreCameras[LoreState], IsCamera := AllEnabledLoreCameras[LoreState], IsCamera?):
                            Character.PutInStasis(stasis_args{AllowEmotes:=false,AllowFalling:=false,AllowTurning:=false})
                            LoreCamera.AddTo(Agent)
                            set MaybeCamera = option{LoreCamera}
                        if(StringLore:=AllJLore[LoreState], StringLore.Length>2):
                            GameManager.DisplayCharacterMessage(Agent,J_CHARACTER,StringLore, ?DisplayTime := 5.0, ?Priority:=MAX_PRIORITY)
                            Sleep(5.1)
                        if(StringLore:=AllWLore[LoreState], StringLore.Length>2):
                            GameManager.DisplayCharacterMessage(Agent,W_CHARACTER,StringLore, ?DisplayTime := 5.0, ?Priority:=MAX_PRIORITY)
                            Sleep(5.0)
                        if(Camera := MaybeCamera?):
                            Camera.RemoveFrom(Agent)
                        Character.ReleaseFromStasis()
                        if(SavedLoreState := AgentBakery.PlayerDataCache.AllLoreStates[MAIN_LORE]):      
                            Log("Posible Lore Finished Now Player has {SavedLoreState} Lore State!", LOG_LEVEL_DEBUG7)
                    else if(LoreState-2 <= AgentBakery.PlayerDataCache.LV_Area):
                        Log("Executing Main Lore!", LOG_LEVEL_DEBUG7)
                        if(Player.IsActive[]):
                            NewLoreState := LoreState + 1
                            AgentBakery.SetNewPlayerDataCache(GameManager.PlayersDataManager.SetLoreState(AgentBakery.PlayerDataCache,MAIN_LORE,NewLoreState))
                        Sleep(PreDelay)
                        
                        var MaybeCamera : ?gameplay_camera_fixed_point_device = false
                        if(LoreCamera := AllLoreCameras[LoreState], IsCamera := AllEnabledLoreCameras[LoreState], IsCamera?):
                            Character.PutInStasis(stasis_args{AllowEmotes:=false,AllowFalling:=false,AllowTurning:=false})
                            LoreCamera.AddTo(Agent)
                            set MaybeCamera = option{LoreCamera}
                        if(StringLore:=AllJLore[LoreState], StringLore.Length>2):
                            GameManager.DisplayCharacterMessage(Agent,J_CHARACTER,StringLore, ?DisplayTime := 5.0, ?Priority:=MAX_PRIORITY)
                            Sleep(5.1)
                        if(StringLore:=AllWLore[LoreState], StringLore.Length>2):
                            GameManager.DisplayCharacterMessage(Agent,W_CHARACTER,StringLore, ?DisplayTime := 5.0, ?Priority:=MAX_PRIORITY)
                            Sleep(5.0)
                        else:
                            PrintWarning("Lore not found at AllWLore[{LoreState}]")
                        if(Camera := MaybeCamera?):
                            Camera.RemoveFrom(Agent)
                        Character.ReleaseFromStasis()

    LoreInteraction(Agent : agent, LoreType : int) : void =
        if(GameManager := MaybeGameManager?, AgentBakery := GameManager.GetAgentBakery_Transacted(Agent)?):
            #NumAreas:=AgentBakery.PlayerDataCache.LV_Area
            var SecretHelpPos : float = 0.0
            var AllStringLore : []string = array{}
            var Character : int = 0
            DisplayTime := 5.0

            if(LoreType=W_RANDOM_LORE):
                set AllStringLore = AllWRandom
                set Character = W_CHARACTER
                set SecretHelpPos = -1.0
            else if(LoreType=H_RANDOM_LORE):
                set AllStringLore = AllHRandom
                set Character = H_CHARACTER
                set SecretHelpPos = 0.5
            else if(LoreType=G_RANDOM_LORE):
                set AllStringLore = AllSecretsTips
                set Character = G_CHARACTER
                set SecretHelpPos = 0.3
            else if(LoreType=J0_RANDOM_LORE):
                set AllStringLore = AllJRandom
                set Character = J_CHARACTER
                set SecretHelpPos = 0.3
            else if(LoreType=J1_RANDOM_LORE):
                set AllStringLore = AllJRandom_SecretBakery
                set Character = J_CHARACTER
                set SecretHelpPos = 0.3
            else if(LoreType=J2_RANDOM_LORE):
                set AllStringLore = AllJRandom_Pure
                set Character = J_CHARACTER
                set SecretHelpPos = 0.3
                

            SecretHelpDecision := GetRandomFloat(0.0,1.0)
            if(SecretHelpDecision < SecretHelpPos):
                var AllSecretsToTell : []int = array{}
                for(Index:=0..20, IsGranted := AgentBakery.PlayerDataCache.AllSecrets_Granted[Index], not IsGranted?, StringHelp:=AllSecretsHelp[Index], StringHelp.Length > 10):
                    Log("Secret {Index} valid because is {StringHelp.Length} Length", LOG_LEVEL_DEBUG7)
                    set AllSecretsToTell = AllSecretsToTell + array{Index}
                if(AllSecretsToTell.Length > 0):
                    IndexValidSecretToTell := GetRandomInt(0,AllSecretsToTell.Length-1)
                    if(RandomSecretIndex:=AllSecretsToTell[IndexValidSecretToTell],FinalStringHelp:=AllSecretsHelp[RandomSecretIndex]):
                        AgentIndex:=GameManager.GetAgentIndex(Agent)
                        if(AccumulatedNumber:=AllAgentLoreAccumulated[AgentIndex]):
                            if(AccumulatedNumber>0):
                                GameManager.CleanCharacterMessages(Agent)
                            GameManager.DisplayCharacterMessage(Agent,Character,FinalStringHelp, ?DisplayTime:=DisplayTime) #QUEST UNLOCKED
                            if(set AllAgentLoreAccumulated[AgentIndex] = AccumulatedNumber + 1):
                            spawn:
                                DelayAccumulatedNumber(AgentIndex,DisplayTime)
                        else:
                            PrintFatalError("On LoreInteraction")
            else:
                if(var CurrentLoreState : int = AgentBakery.PlayerDataCache.AllLoreStates[LoreType]):
                    if(CurrentLoreState < AllStringLore.Length):
                        if(LoreString := AllStringLore[CurrentLoreState]):
                            set CurrentLoreState = CurrentLoreState + 1
                            AgentIndex:=GameManager.GetAgentIndex(Agent)
                            if(AccumulatedNumber:=AllAgentLoreAccumulated[AgentIndex]):
                                if(AccumulatedNumber>0):
                                    GameManager.CleanCharacterMessages(Agent)
                                GameManager.DisplayCharacterMessage(Agent,Character,LoreString, ?DisplayTime:=DisplayTime) #QUEST UNLOCKED
                                if(set AllAgentLoreAccumulated[AgentIndex] = AccumulatedNumber + 1):
                                spawn:
                                    DelayAccumulatedNumber(AgentIndex,DisplayTime)
                    else:
                        set CurrentLoreState = 0
                    AgentBakery.SetNewPlayerDataCache(GameManager.PlayersDataManager.SetLoreState(AgentBakery.PlayerDataCache,LoreType,CurrentLoreState))
                    
                else:
                    PrintFatalError("On LoreInteraction, LoreType {LoreType} INVALID!")

    DelayAccumulatedNumber(AgentIndex : int, Delay : float)<suspends> : void =
        Sleep(Delay)
        if(Amount := AllAgentLoreAccumulated[AgentIndex], set AllAgentLoreAccumulated[AgentIndex] = Amount - 1):
            if(AllAgentLoreAccumulated[AgentIndex] < 0, set AllAgentLoreAccumulated[AgentIndex] = 0):

    CheckQuestInited() : void =
        if(not AllQuestInitiated?):
            InitAllQuests()
            set AllQuestInitiated = true

    GetButtonDevice<private>(NumQuestCompleted : int, NumQuest : int, QuestType :int, Agent : agent) : ?button_device =
        var MaybeButton : ?button_device = false
        if(GameManager := MaybeGameManager?):
            MaybeBakery:=GameManager.GetAgentBakery(Agent)
            if(ValidBakery := MaybeBakery?):
                    Log("QuestType: {QuestType}",LOG_LEVEL_DEBUG7)
                if(QuestType=USEFUNNEL_TUTORIAL_TYPE):
                    Log("USEFUNNEL_TUTORIAL_TYPE",LOG_LEVEL_DEBUG7)
                    set MaybeButton = option{ValidBakery.Button_ResFunnel}
                else if(QuestType=USEVALVES_TUTORIAL_TYPE, ValveButton:=ValidBakery.AllButtons_Valves[0]):
                    Log("USEVALVES_TUTORIAL_TYPE",LOG_LEVEL_DEBUG7)
                    set MaybeButton = option{ValveButton}
                else if(QuestType=CHECKINFO_TUTORIAL_TYPE or QuestType=CHANGEVALVES_QUEST_TYPE):
                    Log("CHECKINFO_TUTORIAL_TYPE",LOG_LEVEL_DEBUG7)
                    set MaybeButton = option{ValidBakery.Button_InfoBB}
                else if(QuestType=B_TYPE or QuestType=CB_TYPE or QuestType=PCS_TYPE):
                    Log("Product type: {QuestType}",LOG_LEVEL_DEBUG7)
                    set MaybeButton = option{ValidBakery.Button_DroneShip}
                else if(QuestType=IMPROVEMAINDEPOSITS_TUTORIAL_TYPE or QuestType=AGGEDSAPPHIRESUSED_TUTORIAL_TYPE):
                    Log("Product type: {QuestType}",LOG_LEVEL_DEBUG7)
                    set MaybeButton = option{ValidBakery.Button_DepositsUpgrades}
                else if(QuestType=IMPROVEBAKERY_TUTORIAL_TYPE):
                    Log("IMPROVEBAKERY_TUTORIAL_TYPE Detected!", LOG_LEVEL_DEBUG7)
                    set MaybeButton = option{ValidBakery.Button_BBUpgrades}
                else if(QuestType=OPENINGVALVES_TUTORIAL_TYPE or QuestType=HALFVALVES_TUTORIAL_TYPE, ValveButton:=ValidBakery.AllButtons_Valves[2]):
                    Log("OPENINGVALVES_TUTORIAL_TYPE Detected!", LOG_LEVEL_DEBUG7)
                    set MaybeButton = option{ValveButton}
                else if(QuestType=CLOSINGVALVES_TUTORIAL_TYPE,ValveButton:=ValidBakery.AllButtons_Valves[0]):
                    Log("CLOSINGVALVES_TUTORIAL_TYPE Detected!", LOG_LEVEL_DEBUG7)
                    set MaybeButton = option{ValveButton}
                else if(QuestType=PCSBOPENINGVALVES_TUTORIAL_TYPE,ValveButton:=ValidBakery.AllButtons_Valves[5]):
                    Log("PCSBOPENINGVALVES_TUTORIAL_TYPE Detected!", LOG_LEVEL_DEBUG7)
                    set MaybeButton = option{ValveButton}
                else: 
                    if(NumQuestCompleted>Quest0_Index):
                        if(QuestTupleData := AllNormalQuests[NumQuest]):
                            Log("Trying to find Button of Quest {NumQuest}", LOG_LEVEL_DEBUG7)
                            set MaybeButton = QuestTupleData(8)
                        else:
                            Log("On GetButtonDevice Quests button not found!", LOG_LEVEL_DEBUG7)
                    else:
                        if(TutorialQuestTupleData := AllTutorialQuests[NumQuest]):
                            Log("Trying to find Button of TutorialQuest {NumQuest}", LOG_LEVEL_DEBUG7)
                            set MaybeButton = TutorialQuestTupleData(8)
                        else:
                            Log("On GetButtonDevice Tutorial button not found!", LOG_LEVEL_DEBUG7)
            else:
                PrintFatalError("Bakery not found!")
        return MaybeButton 


    InitAllQuests<private>() : void =
        if:
            TutorialProp_0:=AllTutorialCreativeProps[0]
            TutorialProp_1:=AllTutorialCreativeProps[1]
            TutorialProp_2:=AllTutorialCreativeProps[2]
            TutorialProp_3:=AllTutorialCreativeProps[3]
            TutorialProp_4:=AllTutorialCreativeProps[4]
            TutorialProp_5:=AllTutorialCreativeProps[5]
            TutorialProp_6:=AllTutorialCreativeProps[6]
            TutorialProp_7:=AllTutorialCreativeProps[7]
            TutorialProp_8:=AllTutorialCreativeProps[8]
            TutorialProp_9:=AllTutorialCreativeProps[9]
            TutorialProp_10:=AllTutorialCreativeProps[10]
            TutorialProp_11:=AllTutorialCreativeProps[11]
            TutorialProp_12:=AllTutorialCreativeProps[12]
            TutorialProp_13:=AllTutorialCreativeProps[13]
            TutorialProp_14:=AllTutorialCreativeProps[14]
            TutorialProp_15:=AllTutorialCreativeProps[15]
            TutorialProp_16:=AllTutorialCreativeProps[16]
            TutorialProp_17:=AllTutorialCreativeProps[17]
            TutorialProp_18:=AllTutorialCreativeProps[18]
            TutorialProp_19:=AllTutorialCreativeProps[19]
            TutorialProp_20:=AllTutorialCreativeProps[20]
            TutorialProp_21:=AllTutorialCreativeProps[21]
            TutorialProp_22:=AllTutorialCreativeProps[22]
            TutorialProp_23:=AllTutorialCreativeProps[23]
            TutorialProp_24:=AllTutorialCreativeProps[24]
            TutorialProp_25:=AllTutorialCreativeProps[25]
            TutorialProp_26:=AllTutorialCreativeProps[26]
            TutorialProp_27:=AllTutorialCreativeProps[27]
            TutorialProp_28:=AllTutorialCreativeProps[28]
            TutorialProp_29:=AllTutorialCreativeProps[29]
            TutorialProp_30:=AllTutorialCreativeProps[30]
            TutorialProp_31:=AllTutorialCreativeProps[31]
            TutorialProp_32:=AllTutorialCreativeProps[32]
            TutorialProp_33:=AllTutorialCreativeProps[33]
            TutorialProp_34:=AllTutorialCreativeProps[34]

            TutorialButton_0:=AllTutorialButtonDevices[0]
            TutorialButton_1:=AllTutorialButtonDevices[1]
            TutorialButton_2:=AllTutorialButtonDevices[2]
            TutorialButton_3:=AllTutorialButtonDevices[3]
            TutorialButton_4:=AllTutorialButtonDevices[4]
            TutorialButton_5:=AllTutorialButtonDevices[5]
            TutorialButton_6:=AllTutorialButtonDevices[6]
            TutorialButton_7:=AllTutorialButtonDevices[7]
            TutorialButton_8:=AllTutorialButtonDevices[8]
            TutorialButton_9:=AllTutorialButtonDevices[9]
            TutorialButton_10:=AllTutorialButtonDevices[10]
            TutorialButton_11:=AllTutorialButtonDevices[11]
            TutorialButton_12:=AllTutorialButtonDevices[12]
            TutorialButton_13:=AllTutorialButtonDevices[13]
            TutorialButton_14:=AllTutorialButtonDevices[14]
            TutorialButton_15:=AllTutorialButtonDevices[15]
            TutorialButton_16:=AllTutorialButtonDevices[16]
            TutorialButton_17:=AllTutorialButtonDevices[17]
            TutorialButton_18:=AllTutorialButtonDevices[18]
            TutorialButton_19:=AllTutorialButtonDevices[19]
            TutorialButton_20:=AllTutorialButtonDevices[20]
            TutorialButton_21:=AllTutorialButtonDevices[21]
            TutorialButton_22:=AllTutorialButtonDevices[22]
            TutorialButton_23:=AllTutorialButtonDevices[23]
            TutorialButton_24:=AllTutorialButtonDevices[24]
            TutorialButton_25:=AllTutorialButtonDevices[25]
            TutorialButton_26:=AllTutorialButtonDevices[26]
            TutorialButton_27:=AllTutorialButtonDevices[27]
            TutorialButton_28:=AllTutorialButtonDevices[28]
            TutorialButton_29:=AllTutorialButtonDevices[29]
            TutorialButton_30:=AllTutorialButtonDevices[30]
            TutorialButton_31:=AllTutorialButtonDevices[31]
            TutorialButton_32:=AllTutorialButtonDevices[32]
            TutorialButton_33:=AllTutorialButtonDevices[33]
            TutorialButton_34:=AllTutorialButtonDevices[34]
            TutorialButton_35:=AllTutorialButtonDevices[35]
            TutorialButton_36:=AllTutorialButtonDevices[36]

                
            QuestProp_0:=AllQuestCreativeProps[0]
            QuestProp_1:=AllQuestCreativeProps[1]
            QuestProp_2:=AllQuestCreativeProps[2]
            QuestProp_3:=AllQuestCreativeProps[3]
            QuestProp_4:=AllQuestCreativeProps[4]
            QuestProp_5:=AllQuestCreativeProps[5]
            QuestProp_6:=AllQuestCreativeProps[6]
            QuestProp_7:=AllQuestCreativeProps[7]
            QuestProp_8:=AllQuestCreativeProps[8]
            QuestProp_9:=AllQuestCreativeProps[9]
            QuestProp_10:=AllQuestCreativeProps[10]
            QuestProp_11:=AllQuestCreativeProps[11]
            QuestProp_12:=AllQuestCreativeProps[12]
            QuestProp_13:=AllQuestCreativeProps[13]
            QuestProp_14:=AllQuestCreativeProps[14]

            QuestButton_0:=AllQuestButtons[0]
            QuestButton_1:=AllQuestButtons[1]
            QuestButton_2:=AllQuestButtons[2]
            QuestButton_3:=AllQuestButtons[3]
            QuestButton_4:=AllQuestButtons[4]
            QuestButton_5:=AllQuestButtons[5]
            QuestButton_6:=AllQuestButtons[6]
            QuestButton_7:=AllQuestButtons[7]
            QuestButton_8:=AllQuestButtons[8]
            QuestButton_9:=AllQuestButtons[9]
            QuestButton_10:=AllQuestButtons[10]
            QuestButton_11:=AllQuestButtons[11]
            QuestButton_12:=AllQuestButtons[12]
            QuestButton_13:=AllQuestButtons[13]
            QuestButton_14:=AllQuestButtons[14]


        then:
            #Tutorial quests
            var Tutorial_0 : tuple(string,texture,string,int,float,int,float,?creative_prop,?button_device) = ("Breaking Bread I",Images.Pickaxe_Icon,"Pickaxe Stale Bread until you get Water",WATER,5.0,FLOUR,100.0,option{TutorialProp_0},option{TutorialButton_0})
            var Tutorial_1 : tuple(string,texture,string,int,float,int,float,?creative_prop,?button_device) = ("Fast travel I",Images.Speed_Icon,"Use the teleporter to travel to your Bakery",USETELEPORTER_TUTORIAL_TYPE,1.0,WATER,100.0,false,option{TutorialButton_1})
            var Tutorial_2 : tuple(string,texture,string,int,float,int,float,?creative_prop,?button_device) = ("Insert the resources",Images.Baker_Icon,"Insert the resources in the funnel",USEFUNNEL_TUTORIAL_TYPE,1.0,SAPPHIRES,1.0,false,option{TutorialButton_2})
            var Tutorial_3 : tuple(string,texture,string,int,float,int,float,?creative_prop,?button_device) = ("Using valves I",Images.Baker_Icon,"Adjust the valves until they are green or yellow",USEVALVES_TUTORIAL_TYPE,1.0,SAPPHIRES,1.0,false,option{TutorialButton_3})
            var Tutorial_3_5 : tuple(string,texture,string,int,float,int,float,?creative_prop,?button_device) = ("Baker Basics",Images.Baker_Icon,"Watch the Baker Tutorial",BAKERTUTORIAL_TUTORIAL_TYPE,1.0,SAPPHIRES,1.0,false,option{TutorialButton_35})
            var Tutorial_4 : tuple(string,texture,string,int,float,int,float,?creative_prop,?button_device) = ("Check the Baker info",Images.Baker_Icon,"Check Bakery info on the computer",CHECKINFO_TUTORIAL_TYPE,1.0,SAPPHIRES,1.0,false,option{TutorialButton_4})
            var Tutorial_5 : tuple(string,texture,string,int,float,int,float,?creative_prop,?button_device) = ("Expanding the business",Images.Pickaxe_Icon,"Pay the fee to unlock the next area",UNLOCKAREA_TUTORIAL_TYPE,1.0,SAPPHIRES,1.0,false,option{TutorialButton_5})
            var Tutorial_6 : tuple(string,texture,string,int,float,int,float,?creative_prop,?button_device) = ("Breaking Bread II",Images.Pickaxe_Icon,"Pickaxe Stale Bread until you get Water",WATER,15.0,WATER,250.0,option{TutorialProp_6},option{TutorialButton_6})
            var Tutorial_7 : tuple(string,texture,string,int,float,int,float,?creative_prop,?button_device) = ("Breaking Bread III",Images.Pickaxe_Icon,"Pickaxe Stale Bread until you get Flour",FLOUR,15.0,FLOUR,250.0,option{TutorialProp_7},option{TutorialButton_7})
            var Tutorial_8 : tuple(string,texture,string,int,float,int,float,?creative_prop,?button_device) = ("Baking Bread I",Images.Bread_Icon,"Bake Bread and send it using the drone",B_TYPE,1.0,SAPPHIRES,5.0,false,option{TutorialButton_8})
            var Tutorial_9 : tuple(string,texture,string,int,float,int,float,?creative_prop,?button_device) = ("Taking a break",Images.Bread_Icon,"Visit the rest room",USEAFKZONE_TUTORIAL_TYPE,1.0,SAPPHIRES,1.0,option{TutorialProp_9},option{TutorialButton_9})
            var Tutorial_10 : tuple(string,texture,string,int,float,int,float,?creative_prop,?button_device) = ("Breaking Harder I",Images.Pickaxe_Icon,"Improve your Pickaxe Damage",IMPROVEPICKAXE_TUTORIAL_TYPE,1.0,SAPPHIRES,5.0,false,option{TutorialButton_10})
            var Tutorial_11 : tuple(string,texture,string,int,float,int,float,?creative_prop,?button_device) = ("Upgrade your backpack",Images.Pickaxe_Icon,"Upgrade your backpack in the store",IMPROVEBACKPACK_TUTORIAL_TYPE,1.0,SAPPHIRES,5.0,false,option{TutorialButton_11})
            var Tutorial_12 : tuple(string,texture,string,int,float,int,float,?creative_prop,?button_device) = ("Upgrade your deposits",Images.Pickaxe_Icon,"Upgrade your Bakery deposits",IMPROVEMAINDEPOSITS_TUTORIAL_TYPE,1.0,SAPPHIRES,5.0,false,option{TutorialButton_12})
            var Tutorial_13 : tuple(string,texture,string,int,float,int,float,?creative_prop,?button_device) = ("Upgrade your Bakery",Images.Bread_Icon,"Upgrade your Bread Bakery production",IMPROVEBAKERY_TUTORIAL_TYPE,1.0,SAPPHIRES,5.0,false,option{TutorialButton_13})
            var Tutorial_14 : tuple(string,texture,string,int,float,int,float,?creative_prop,?button_device) = ("Baking Bread II",Images.Bread_Icon,"Bake more Bread and send it using the drone",B_TYPE,10.0,SAPPHIRES,10.0,false,option{TutorialButton_14})
            var Tutorial_15 : tuple(string,texture,string,int,float,int,float,?creative_prop,?button_device) = ("Breaking Harder II",Images.Pickaxe_Icon,"Upgrade your Pickaxe perks in the store",IMPROVEPICKPERK_TUTORIAL_TYPE,1.0,SAPPHIRES,10.0,false,option{TutorialButton_15})
            var Tutorial_16 : tuple(string,texture,string,int,float,int,float,?creative_prop,?button_device) = ("Breaking Bread IV",Images.Pickaxe_Icon,"Break 5 Stale Breads",BREAKBREAD_TUTORIAL_TYPE,5.0,SAPPHIRES,10.0,option{TutorialProp_16},option{TutorialButton_16})
            var Tutorial_17 : tuple(string,texture,string,int,float,int,float,?creative_prop,?button_device) = ("Using valves II",Images.Baker_Icon,"Set Cheese Bread Baker valves to green color",OPENINGVALVES_TUTORIAL_TYPE,1.0,SAPPHIRES,10.0,false,option{TutorialButton_17})
            var Tutorial_18 : tuple(string,texture,string,int,float,int,float,?creative_prop,?button_device) = ("Using valves III",Images.Baker_Icon,"Close all valves on your Bread Baker",CLOSINGVALVES_TUTORIAL_TYPE,1.0,SAPPHIRES,10.0,false,option{TutorialButton_18})
            var Tutorial_19 : tuple(string,texture,string,int,float,int,float,?creative_prop,?button_device) = ("Using valves IV",Images.Baker_Icon,"Set ALL valves to half open (Yellow) on ALL Bakers",HALFVALVES_TUTORIAL_TYPE,1.0,SAPPHIRES,10.0,false,option{TutorialButton_19})
            var Tutorial_20 : tuple(string,texture,string,int,float,int,float,?creative_prop,?button_device) = ("Breaking Bread V",Images.Pickaxe_Icon,"Pickaxe Stale Bread until you get Cheese",CHEESE,500.0,COAL,5.0,option{TutorialProp_20},option{TutorialButton_20})
            var Tutorial_21 : tuple(string,texture,string,int,float,int,float,?creative_prop,?button_device) = ("Baking Bread III",Images.CheeseBread_Icon,"Bake Cheese Bread and send it using the drone",CB_TYPE,25.0,COAL,5.0,false,option{TutorialButton_21})
            var Tutorial_22 : tuple(string,texture,string,int,float,int,float,?creative_prop,?button_device) = ("Using valves V",Images.PimentoCheeseSandwich_Icon,"Set Pimento Cheese Sandwich Baker to open",PCSBOPENINGVALVES_TUTORIAL_TYPE,1.0,COAL,5.0,false,option{TutorialButton_22})
            var Tutorial_23 : tuple(string,texture,string,int,float,int,float,?creative_prop,?button_device) = ("Using Coal I",Images.Coal_Icon,"Introduce 5 Coal in the Coal Generator",USECOAL_TUTORIAL_TYPE,5.0,COAL,5.0,false,option{TutorialButton_23})
            var Tutorial_24 : tuple(string,texture,string,int,float,int,float,?creative_prop,?button_device) = ("Stay out of my territory",Images.Pickaxe_Icon,"Pay the fee to unlock the next area",UNLOCKAREA_TUTORIAL_TYPE,1.0,COAL,5.0,false,option{TutorialButton_24})
            var Tutorial_24_5 : tuple(string,texture,string,int,float,int,float,?creative_prop,?button_device) = ("Meet the Team",Images.Icon_Druid,"Interact with a Sourdough on the table to meet it",PETTUTORIAL_TUTORIAL_TYPE,1.0,MUTANTMATERIAL,25.0,false,option{TutorialButton_36})
            var Tutorial_25 : tuple(string,texture,string,int,float,int,float,?creative_prop,?button_device) = ("Purest Formula I",Images.Mutant_Material_Icon,"Create a Sourdough using Mutant Material in the Sourdough Machine",CREATEPET_TUTORIAL_TYPE,1.0,MUTANTMATERIAL,5.0,false,option{TutorialButton_25})
            var Tutorial_26 : tuple(string,texture,string,int,float,int,float,?creative_prop,?button_device) = ("Purest Formula II",Images.Coal_Icon,"Use Coal to improve a Sourdough Pack's stats",IMPROVEPET_TUTORIAL_TYPE,1.0,COAL,5.0,false,option{TutorialButton_26})
            var Tutorial_27 : tuple(string,texture,string,int,float,int,float,?creative_prop,?button_device) = ("Purest Formula III",Images.Icon_ADC,"Unlock a Sourdough Competition Challenge",UNLOCKHEXAGAME_TUTORIAL_TYPE,1.0,COAL,5.0,option{TutorialProp_27},option{TutorialButton_27})
            var Tutorial_28 : tuple(string,texture,string,int,float,int,float,?creative_prop,?button_device) = ("Fast travel II",Images.Speed_Icon,"Use Coal on the teleporter to travel to your Bakery",USEEXPENSIVETELEPORTER_TUTORIAL_TYPE,1.0,COAL,5.0,false,option{TutorialButton_28})
            var Tutorial_29 : tuple(string,texture,string,int,float,int,float,?creative_prop,?button_device) = ("Purest Formula IV",Images.Icon_ADC,"Play and win a Sourdough Competition Challenge",WINHEXAGAME_TUTORIAL_TYPE,1.0,COAL,5.0,false,option{TutorialButton_29})
            var Tutorial_30 : tuple(string,texture,string,int,float,int,float,?creative_prop,?button_device) = ("Complete an order",Images.PimentoCheeseSandwich_Icon,"Send Pimento Cheese Sandwiches using the drone",PCS_TYPE,150.0,COAL,5.0,false,option{TutorialButton_30})
            var Tutorial_32 : tuple(string,texture,string,int,float,int,float,?creative_prop,?button_device) = ("Weapon improve",Images.Sword_Icon,"Upgrade the Hammer in the store",WEAPONIMPROVE_TUTORIAL_TYPE,1.0,COAL,5.0,false,option{TutorialButton_32})
            var Tutorial_31 : tuple(string,texture,string,int,float,int,float,?creative_prop,?button_device) = ("The bad batch",Images.Sword_Icon,"Clear a dungeon and destroy the mutant bread",COMPLETEDUNGEON_TUTORIAL_TYPE,1.0,COAL,5.0,false,option{TutorialButton_31})
            #var Tutorial_33 : tuple(string,texture,string,int,float,int,float,?creative_prop,?button_device) = ("Improve a token",Images.Challenge_Icon,"Upgrade a Sourdough Competition or Dungeon Token",TOKENIMPROVE_TUTORIAL_TYPE,1.0,AGEDSAPPHIRES,5.0,false,option{TutorialButton_33})
            #var Tutorial_34 : tuple(string,texture,string,int,float,int,float,?creative_prop,?button_device) = ("Use Aged Sapphires",Images.AgedSaphire_Icon,"Use Aged Sapphires to upgrade the Dungeon Token",AGGEDSAPPHIRESUSED_TUTORIAL_TYPE,1.0,AGEDSAPPHIRES,5.0,false,option{TutorialButton_34})
            
            Quest0 : tuple(string,texture,string,int,float,int,float,?creative_prop,?button_device) = ("The Supreme Formula",ParkourIcon,"Find the ancient formula near the Bakery factory area",FINDSCROLL_AREA0_QUEST_TYPE,1.0,COAL,10.0,option{QuestProp_0},option{QuestButton_0})

            #Quests
            Quest_0 : tuple(string,texture,string,int,float,int,float,?creative_prop,?button_device) = ("Show your formula",ParkourIcon,"Play and win a Sourdough Competition Challenge",WINHEXAGAME_TUTORIAL_TYPE,1.0,COAL,StandarQuestReward_Coal,false,option{TutorialButton_29})
            Quest_1 : tuple(string,texture,string,int,float,int,float,?creative_prop,?button_device) = ("Teleport Testing",ParkourIcon,"Spend Coal on teleporters",USEEXPENSIVETELEPORTER_TUTORIAL_TYPE,10.0,COAL,StandarQuestReward_Coal+10.0,false,option{TutorialButton_28})
            Quest_2 : tuple(string,texture,string,int,float,int,float,?creative_prop,?button_device) = ("Coffee time",ParkourIcon,"Spend 10 seconds in the rest area",USEAFKZONE_TUTORIAL_TYPE,10.0,COAL,StandarQuestReward_Coal,option{TutorialProp_9},option{TutorialButton_9})
            Quest_3 : tuple(string,texture,string,int,float,int,float,?creative_prop,?button_device) = ("Quick check",ParkourIcon,"Check your Bakery info on the computers",CHECKINFO_TUTORIAL_TYPE,3.0,COAL,StandarQuestReward_Coal,false,option{TutorialButton_4})
            Quest_4 : tuple(string,texture,string,int,float,int,float,?creative_prop,?button_device) = ("Community service",Images.Sword_Icon,"Clear any Dungeon and destroy the mutant bread",COMPLETEDUNGEON_TUTORIAL_TYPE,1.0,COAL,StandarQuestReward_Coal,false,option{TutorialButton_31})
            Quest_5 : tuple(string,texture,string,int,float,int,float,?creative_prop,?button_device) = ("Pressing buttons",Images.Sword_Icon,"Adjust the Baker valve settings 10 times",CHANGEVALVES_QUEST_TYPE,10.0,COAL,StandarQuestReward_Coal,false,false)
            Quest_6 : tuple(string,texture,string,int,float,int,float,?creative_prop,?button_device) = ("The Supreme Formula",ParkourIcon,"Find the ancient formula near the Bakery factory area",FINDSCROLL_AREA0_QUEST_TYPE,1.0,COAL,StandarQuestReward_Coal,false,false)
            Quest_7 : tuple(string,texture,string,int,float,int,float,?creative_prop,?button_device) = ("The Supreme Formula",ParkourIcon,"Find the ancient formula in the Chicken Lovers restaurant area",FINDSCROLL_AREA2_QUEST_TYPE,1.0,COAL,StandarQuestReward_Coal,false,false)
            Quest_8 : tuple(string,texture,string,int,float,int,float,?creative_prop,?button_device) = ("The Supreme Formula",ParkourIcon,"Find the ancient formula near the Broken camper",FINDSCROLL_AREA8_QUEST_TYPE,1.0,COAL,StandarQuestReward_Coal,false,false)
            Quest_9 : tuple(string,texture,string,int,float,int,float,?creative_prop,?button_device) = ("Electrical maintenance",Images.Challenge_Icon,"Repair the factory's electrical system",ELECTRICALSYSTEM_MINIGAME_QUEST_TYPE,1.0,COAL,StandarQuestReward_Coal,false,option{QuestButton_9})
            Quest_10 : tuple(string,texture,string,int,float,int,float,?creative_prop,?button_device) = ("Pressure maintenance",Images.Challenge_Icon,"Repair the factory's pressure system",PREASSURESYSTEM_MINIGAME_QUEST_TYPE,1.0,COAL,StandarQuestReward_Coal,false,option{QuestButton_10})
            Quest_11 : tuple(string,texture,string,int,float,int,float,?creative_prop,?button_device) = ("Truck maintenance",Images.Challenge_Icon,"Repair the factory's truck",TRUCKENGINE_MINIGAME_QUEST_TYPE,1.0,COAL,StandarQuestReward_Coal,false,option{QuestButton_11})
            Quest_12 : tuple(string,texture,string,int,float,int,float,?creative_prop,?button_device) = ("Playing with fire",Images.Challenge_Icon,"Hit the unstable Slate Bread",UNSTABLEHB_QUEST_TYPE,15.0,COAL,StandarQuestReward_Coal,false,false)
            Quest_13 : tuple(string,texture,string,int,float,int,float,?creative_prop,?button_device) = ("Helping a friend",Images.Challenge_Icon,"Perform maintenance on any of the Steve Baking Machines",STEVEBAKER_QUEST_TYPE,1.0,COAL,StandarQuestReward_Coal,false,option{QuestButton_13})
            QuestSPECIAL_Mastery : tuple(string,texture,string,int,float,int,float,?creative_prop,?button_device) = ("Bake Break Reborn",Images.Challenge_Icon,"Sell your Bakery to earn Aged Sapphires",MASTERY_QUEST_TYPE,1.0,AGEDSAPPHIRES,5.0,false,option{QuestButton_14})
            
            var DummyQuest : tuple(string,texture,string,int,float,int,float,?creative_prop,?button_device) = ("",CheeseIcon,"",-1,-1.0,-1,-1.0,false,false)
            for(Index:=0..100):
                set AllTutorialQuests = AllTutorialQuests + array{DummyQuest}
            for(Index:=0..100):
                set AllNormalQuests = AllNormalQuests + array{DummyQuest}

            #Tutorial Quests
            if:
                set AllTutorialQuests[0] = Tutorial_0
                set AllTutorialQuests[1] = Tutorial_1
                set AllTutorialQuests[2] = Tutorial_2
                set AllTutorialQuests[3] = Tutorial_3
                set AllTutorialQuests[4] = Tutorial_3_5
                set AllTutorialQuests[5] = Tutorial_4
                set AllTutorialQuests[6] = Tutorial_5
                set AllTutorialQuests[7] = Tutorial_6
                set AllTutorialQuests[8] = Tutorial_7
                set AllTutorialQuests[9] = Tutorial_8
                set AllTutorialQuests[10] = Tutorial_9
                set AllTutorialQuests[11] = Tutorial_10
                set AllTutorialQuests[12] = Tutorial_11
                set AllTutorialQuests[13] = Tutorial_12
                set AllTutorialQuests[14] = Tutorial_13
                set AllTutorialQuests[15] = Tutorial_14
                set AllTutorialQuests[16] = Tutorial_15
                set AllTutorialQuests[17] = Tutorial_16
                set AllTutorialQuests[18] = Tutorial_17
                set AllTutorialQuests[19] = Tutorial_18
                set AllTutorialQuests[20] = Tutorial_19
                set AllTutorialQuests[21] = Tutorial_20
                set AllTutorialQuests[22] = Tutorial_21
                set AllTutorialQuests[23] = Tutorial_22
                set AllTutorialQuests[24] = Tutorial_23
                set AllTutorialQuests[25] = Tutorial_24
                set AllTutorialQuests[26] = Tutorial_24_5
                set AllTutorialQuests[27] = Tutorial_25
                set AllTutorialQuests[28] = Tutorial_26
                set AllTutorialQuests[29] = Tutorial_27
                set AllTutorialQuests[30] = Tutorial_28
                set AllTutorialQuests[31] = Tutorial_29
                set AllTutorialQuests[32] = Tutorial_30
                set AllTutorialQuests[33] = Tutorial_31
                set AllTutorialQuests[34] = Tutorial_32
                #set AllTutorialQuests[35] = Tutorial_33
                #set AllTutorialQuests[36] = Tutorial_34


                set AllTutorialQuests[36] = Quest0
            else:
                #Print("Fatal error on InitAllQuests: AllTutorialQuests")
            #Quests
            if:
                set AllNormalQuests[0] = Quest_0
                set AllNormalQuests[1] = Quest_1
                set AllNormalQuests[2] = Quest_2
                set AllNormalQuests[3] = Quest_3
                set AllNormalQuests[4] = Quest_4
                set AllNormalQuests[5] = Quest_5
                set AllNormalQuests[6] = Quest_3 #Collectible device does not work
                set AllNormalQuests[7] = Quest_4 #Collectible device does not work
                set AllNormalQuests[8] = Quest_5 #Collectible device does not work
                set AllNormalQuests[9] = Quest_9
                set AllNormalQuests[10] = Quest_10
                set AllNormalQuests[11] = Quest_11
                set AllNormalQuests[12] = Quest_12
                set AllNormalQuests[13] = Quest_13
                set AllNormalQuests[MASTERYSPECIALQUEST_INDEX] = QuestSPECIAL_Mastery
            else:
                #Print("Fatal error on InitAllQuests: AllNormalQuests")

            #Super Special Quests
        else:
            #Print("Fatal error on InitAllQuests. AllTutorialCreativeProps.Length is {AllTutorialCreativeProps.Length}, AllQuestCreativeProps.Length is {AllQuestCreativeProps.Length}")

    AddAgentQuest_ProgressV2(Player : player, InPlayerData : player_data_table, ProgressAmount : float, ObjetiveType : int) : tuple(logic,logic,float,int,player_data_table) =
        var QuestIsCompleted : logic = false 
        var UIMustBeUpdated : logic = false
        var AwardAmountToDisplay : float = -1.0
        var AwardTypeToDisplay : int = -1
        var PlayerDataTable : player_data_table = InPlayerData
        if(PlayerDataTable.QuestAssigned_NumQuest <> -99, GameManager:=MaybeGameManager?, Agent := agent[Player], ValidBakery := GameManager.GetAgentBakery_Transacted(Agent)?, not ValidBakery.IsOnUKTutorial?): #Is the quest valid to complete?
            Log("Objetive quest: {PlayerDataTable.QuestAssigned_ObjetiveType}, you just completed a task Type: {ObjetiveType}",LOG_LEVEL_DEBUG6)
            if(ObjetiveType=PlayerDataTable.QuestAssigned_ObjetiveType): #Is the objetive the same?
                set UIMustBeUpdated = true
                QuestInfoTuple:=GetCurrentQuest(PlayerDataTable.NumQuestsCompleted,PlayerDataTable)
                NewProgress:=PlayerDataTable.QuestAssigned_ProgressAmount + ProgressAmount
                if(QuestInfoTuple(4)<=NewProgress):
                    #Print("Quest completed!")
                    set QuestIsCompleted = true
                    set AwardAmountToDisplay = PlayerDataTable.QuestAssigned_Reward
                    set AwardTypeToDisplay = PlayerDataTable.QuestAssigned_RewardType
                    #Only for resourcers limited
                    var BackpackLimit : float = 0.0
                    MaybeBakery:=GameManager.GetAgentBakery(Agent)
                    if(Value:=ValidBakery.AllBackpackLimits[PlayerDataTable.QuestAssigned_RewardType],set BackpackLimit = Value):
                    #Continue
                    if(ObjetiveType=MASTERY_QUEST_TYPE):
                        set PlayerDataTable = PlayersDataManager.SetMasteryQuestCompleted(PlayerDataTable,true)
                    set PlayerDataTable = PlayersDataManager.AddResourcers(PlayerDataTable,PlayerDataTable.QuestAssigned_Reward,PlayerDataTable.QuestAssigned_RewardType,BackpackLimit)(0)
                    set PlayerDataTable = QuestCompletedEvent(PlayerDataTable,Player) #New quest created
                else:
                    #Save current progress
                    #Print("Quest progress added! {NewProgress} / {QuestInfoTuple(4)}")
                    set QuestIsCompleted = false
                    set PlayerDataTable = PlayersDataManager.SetQuest_Progress(PlayerDataTable,NewProgress)
                if(MainInventoryUI:=ValidBakery.MaybeMainInventoryUI?, ActionUI:=ValidBakery.MaybeActionInfoUI?):
                    if(QuestIsCompleted?):
                        Log("Quest Reward is {AwardAmountToDisplay}, Quest Reward Type is {AwardTypeToDisplay}",LOG_LEVEL_TUTORIALQUESTS)
                        ActionUI.DisplayGather(array{AwardAmountToDisplay}, array{AwardTypeToDisplay}, array{false},false)
                        GameManager.VFXManager.BigObjetive_AudioPlayer.Play(Agent)
                        spawn:
                            ValidBakery.Delay_NewQuestDisplay(Player)
                    else if(UIMustBeUpdated?):
                        MainInventoryUI.RefreshOrdersQuests(PlayerDataTable)
                    if(UIMustBeUpdated?):
                        ValidBakery.RefreshInventoryUI(PlayerDataTable)
                set PlayerDataTable = CheckIfTutorialWasCompletedBefore(Player,PlayerDataTable)

        else:
            PrintWarning("QuestAssigned_NumQuest: {PlayerDataTable.QuestAssigned_NumQuest} not valid")
        return (QuestIsCompleted,UIMustBeUpdated,AwardAmountToDisplay,AwardTypeToDisplay,PlayerDataTable)

        <#
        AddAgentQuest_Progress(Player : player, PlayerData : player_data_table, ProgressAmount : float, ObjetiveType : int) : tuple(logic,logic,float,int,player_data_table) =
            var QuestIsCompleted : logic = false 
            var UIMustBeUpdated : logic = false
            var AwardAmountToDisplay : float = -1.0
            var AwardTypeToDisplay : int = -1
            var UpdatedDataTable : player_data_table = PlayerData
            if(PlayerData.QuestAssigned_NumQuest <> -99): #Is the quest valid to complete?
                ##Print("Objetive quest: {PlayerData.QuestAssigned_ObjetiveType}, you just completed a task Type: {ObjetiveType}")
                if(ObjetiveType=PlayerData.QuestAssigned_ObjetiveType): #Is the objetive the same?
                    set UIMustBeUpdated = true
                    QuestInfoTuple:=GetQuest(PlayerData.QuestAssigned_NumQuest)
                    NewProgress:=PlayerData.QuestAssigned_ProgressAmount + ProgressAmount
                    if(QuestInfoTuple(4)<=NewProgress):
                        #Print("Quest completed!")
                        set QuestIsCompleted = true
                        set AwardAmountToDisplay = PlayerData.QuestAssigned_Reward
                        set AwardTypeToDisplay = PlayerData.QuestAssigned_RewardType
                        #Only for resourcers limited
                        var BackpackLimit : float = 0.0
                        if(GameManager:=MaybeGameManager?, Agent := agent[Player]):
                            MaybeBakery:=GameManager.GetAgentBakery(Agent)
                            if(ValidBakery:=MaybeBakery?, Value:=ValidBakery.AllBackpackLimits[PlayerData.QuestAssigned_RewardType],set BackpackLimit = Value):
                        #Continue
                        set UpdatedDataTable = PlayersDataManager.AddResourcers(UpdatedDataTable,PlayerData.QuestAssigned_Reward,PlayerData.QuestAssigned_RewardType,BackpackLimit)(0)
                        set UpdatedDataTable = QuestCompletedEvent(UpdatedDataTable,Player) #New quest created
                    else:
                        #Save current progress
                        #Print("Quest progress added! {NewProgress} / {QuestInfoTuple(4)}")
                        set QuestIsCompleted = false
                        set UpdatedDataTable = PlayersDataManager.SetQuest_Progress(PlayerData,NewProgress)
                    #PlayersDataManager.SaveNewData(Player,PlayerData,UpdatedDataTable)
                    if(GameManager:=MaybeGameManager?, Agent := agent[Player]):
                        MaybeBakery:=GameManager.GetAgentBakery(Agent)
                        if(ValidBakery:=MaybeBakery?):
                            ValidBakery.SetNewPlayerDataCache(UpdatedDataTable)
                if(GameManager:=MaybeGameManager?, Agent :=agent[Player]):
                    MaybeBakeryManager:=GameManager.GetAgentBakery(Agent)
                    if(BakeryManager := MaybeBakeryManager?, MainInventoryUI:=BakeryManager.MaybeMainInventoryUI?, ActionUI:=BakeryManager.MaybeActionInfoUI?):
                        if(QuestIsCompleted?):
                            Log("Quest Reward is {AwardAmountToDisplay}, Quest Reward Type is {AwardTypeToDisplay}",LOG_LEVEL_TUTORIALQUESTS)
                            ActionUI.DisplayGather(array{AwardAmountToDisplay}, array{AwardTypeToDisplay}, array{false},false)
                            GameManager.VFXManager.BigObjetive_AudioPlayer.Play(Agent)
                            spawn:
                                BakeryManager.Delay_NewQuestDisplay(Player)
                        else if(UIMustBeUpdated?):
                            MainInventoryUI.RefreshOrdersQuests(UpdatedDataTable)
                        if(UIMustBeUpdated?):
                            BakeryManager.RefreshInventoryUI(UpdatedDataTable)
                CheckIfTutorialWasCompletedBefore(Player,UpdatedDataTable)
            return (QuestIsCompleted,UIMustBeUpdated,AwardAmountToDisplay,AwardTypeToDisplay,UpdatedDataTable)#>


    AddAgentOrder_Progress(Player : player, PlayerData : player_data_table,  B_On_Drone : float, CB_On_Drone : float, PCS_On_Drone : float) : tuple(logic,logic,float,int,player_data_table) =
        var OrderIsCompleted : logic = false 
        var UIMustBeUpdated : logic = false 
        var AwardAmountToDisplay : float = -1.0
        var AwardTypeToDisplay : int = -1
        var UpdatedDataTable : player_data_table = PlayerData
        if(PlayerData.OrderAssigned_ProductRequested<>-99): #Is the quest valid to complete?
            var NewProgress : float = PlayerData.OrderAssigned_ProgressAmount
            if(B_TYPE=PlayerData.OrderAssigned_ProductRequested): #Is the objetive the same?
                set NewProgress = PlayerData.OrderAssigned_ProgressAmount + B_On_Drone
                set UIMustBeUpdated = true
            if(CB_TYPE=PlayerData.OrderAssigned_ProductRequested): #Is the objetive the same?
                set NewProgress = PlayerData.OrderAssigned_ProgressAmount + CB_On_Drone
                set UIMustBeUpdated = true
            if(PCS_TYPE=PlayerData.OrderAssigned_ProductRequested): #Is the objetive the same?
                set NewProgress = PlayerData.OrderAssigned_ProgressAmount + PCS_On_Drone
                set UIMustBeUpdated = true

            if(PlayerData.OrderAssigned_ProductAmountRequested<=NewProgress):
                set OrderIsCompleted = true
                set AwardAmountToDisplay = PlayerData.OrderAssigned_Reward
                set AwardTypeToDisplay = PlayerData.OrderAssigned_RewardType
                set UpdatedDataTable = PlayersDataManager.AddResourcers(UpdatedDataTable,PlayerData.OrderAssigned_Reward,PlayerData.OrderAssigned_RewardType,-1.0)(0)
                set UpdatedDataTable = OrderCompletedEvent(UpdatedDataTable,Player) #New quest created
            else:
                set OrderIsCompleted = false
                set UpdatedDataTable = PlayersDataManager.SetOrder_Progress(PlayerData,NewProgress)
            #PlayersDataManager.SaveNewData(Player,PlayerData,UpdatedDataTable)

            if(GameManager:=MaybeGameManager?, Agent := agent[Player]):
                MaybeBakery:=GameManager.GetAgentBakery(Agent)
                if(ValidBakery:=MaybeBakery?):
                    ValidBakery.SetNewPlayerDataCache(UpdatedDataTable)
                    #GameManager.PlayersDataManager.SaveAllNewData(Player,PlayerData,UpdatedDataTable, ValidBakery)

        return (OrderIsCompleted,UIMustBeUpdated,AwardAmountToDisplay,AwardTypeToDisplay,UpdatedDataTable)

    CheckAgentOrder_Progress(Player : player, PlayerData : player_data_table, B_On_Drone : float, CB_On_Drone : float, PCS_On_Drone : float) : tuple(logic,logic,player_data_table) =
        var OrderIsCompleted : logic = false 
        var UIMustBeUpdated : logic = false 
        var PlayerDataTable : player_data_table = PlayerData
        if(PlayerData.OrderAssigned_ProductRequested<>-99):
            var NewAmount : float = 0.0
            if(PlayerData.OrderAssigned_ProductRequested=B_TYPE):
                set NewAmount = PlayerData.OrderAssigned_ProgressAmount + B_On_Drone
            else if(PlayerData.OrderAssigned_ProductRequested=CB_TYPE):
                set NewAmount = PlayerData.OrderAssigned_ProgressAmount + CB_On_Drone
            else if(PlayerData.OrderAssigned_ProductRequested=PCS_TYPE):
                set NewAmount = PlayerData.OrderAssigned_ProgressAmount + PCS_On_Drone

            if(NewAmount>=PlayerData.OrderAssigned_ProductAmountRequested):
                set OrderIsCompleted = true
                PrintWarning("TO DO CheckAgentOrder_Progress: On return, UpdateMainUI, Create order delay")
            else:
                #Saving progress
                set OrderIsCompleted = false
                NewData:=PlayersDataManager.SetOrder_Progress(PlayerData,NewAmount)
                set PlayerDataTable = NewData
        return (OrderIsCompleted,UIMustBeUpdated,PlayerDataTable)
    