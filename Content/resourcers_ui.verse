using { /UnrealEngine.com/Temporary/UI }
using { /UnrealEngine.com/Temporary/SpatialMath }
using { /Verse.org/Simulation }
using { /Fortnite.com/Devices }
using { /Fortnite.com/UI }
using { /Verse.org/Colors/NamedColors }
using { /Verse.org/Colors }

# See https://dev.epicgames.com/documentation/en-us/uefn/create-your-own-device-in-verse for how to create a verse device.

# A Verse-authored creative device that can be placed in a level
resourcers_ui := class():

    #Class settings
    var MaybeAgent : ?agent = false

    BakeryMachinesManager : bakery_machines_manager

    RColumnSizeX : float = 200.0
    RColumnSizeY : float = 50.0

    IconPadding:margin=margin{Left:=5.0,Top:=5.0,Right:=5.0,Bottom:=5.0}
    TextPadding:margin=margin{Left:=60.0,Top:=5.0,Right:=5.0,Bottom:=5.0}

    Background_TextureBlock : texture_block = texture_block{DefaultDesiredSize:=vector2{X:=IconSmallSizeX,Y:=IconSmallSizeY}, DefaultImage:=Images.Img_LineBackground}

    Sapphires_TextureBlock : texture_block = texture_block{DefaultDesiredSize:=vector2{X:=IconSmallSizeX,Y:=IconSmallSizeY}, DefaultImage:=Images.Sapphire_Icon}
    AgedSapphires_TextureBlock : texture_block = texture_block{DefaultDesiredSize:=vector2{X:=IconSmallSizeX,Y:=IconSmallSizeY}, DefaultImage:=Images.Sapphire_Icon}
    Fame_TextureBlock : texture_block = texture_block{DefaultDesiredSize:=vector2{X:=IconSmallSizeX,Y:=IconSmallSizeY}, DefaultImage:=Images.Water_Icon}
    Water_TextureBlock : texture_block = texture_block{DefaultDesiredSize:=vector2{X:=IconSmallSizeX,Y:=IconSmallSizeY}, DefaultImage:=Images.Water_Icon}
    Flour_TextureBlock : texture_block = texture_block{DefaultDesiredSize:=vector2{X:=IconSmallSizeX,Y:=IconSmallSizeY}, DefaultImage:=Images.Flour_Icon}
    Cheese_TextureBlock : texture_block = texture_block{DefaultDesiredSize:=vector2{X:=IconSmallSizeX,Y:=IconSmallSizeY}, DefaultImage:=Images.Cheese_Icon}
    Pimentos_TextureBlock : texture_block = texture_block{DefaultDesiredSize:=vector2{X:=IconSmallSizeX,Y:=IconSmallSizeY}, DefaultImage:=Images.Pimento_Icon}
    Mutant_TextureBlock : texture_block = texture_block{DefaultDesiredSize:=vector2{X:=IconSmallSizeX,Y:=IconSmallSizeY}, DefaultImage:=Images.Mutant_Material_Icon}
    Coal_TextureBlock : texture_block = texture_block{DefaultDesiredSize:=vector2{X:=IconSmallSizeX,Y:=IconSmallSizeY}, DefaultImage:=Images.Coal_Icon}

    Sapphires_BackgroundTextureBlock : texture_block = texture_block{DefaultDesiredSize:=vector2{X:=IconSmallSizeX,Y:=IconSmallSizeY}, DefaultImage:=Images.Sapphire_Icon}
    AgedSapphires_BackgroundTextureBlock : texture_block = texture_block{DefaultDesiredSize:=vector2{X:=IconSmallSizeX,Y:=IconSmallSizeY}, DefaultImage:=Images.Sapphire_Icon}
    Water_BackgroundTextureBlock : texture_block = texture_block{DefaultDesiredSize:=vector2{X:=IconSmallSizeX,Y:=IconSmallSizeY}, DefaultImage:=Images.Water_Icon}
    Flour_BackgroundTextureBlock : texture_block = texture_block{DefaultDesiredSize:=vector2{X:=IconSmallSizeX,Y:=IconSmallSizeY}, DefaultImage:=Images.Flour_Icon}
    Cheese_BackgroundTextureBlock : texture_block = texture_block{DefaultDesiredSize:=vector2{X:=IconSmallSizeX,Y:=IconSmallSizeY}, DefaultImage:=Images.Cheese_Icon}
    Pimentos_BackgroundTextureBlock : texture_block = texture_block{DefaultDesiredSize:=vector2{X:=IconSmallSizeX,Y:=IconSmallSizeY}, DefaultImage:=Images.Pimento_Icon}
    Mutant_BackgroundTextureBlock : texture_block = texture_block{DefaultDesiredSize:=vector2{X:=IconSmallSizeX,Y:=IconSmallSizeY}, DefaultImage:=Images.Mutant_Material_Icon}
    Coal_BackgroundTextureBlock : texture_block = texture_block{DefaultDesiredSize:=vector2{X:=IconSmallSizeX,Y:=IconSmallSizeY}, DefaultImage:=Images.Coal_Icon}

    R_Row1_Sapphires : text_block = text_block{DefaultTextColor:=SapphiresTextColor, DefaultShadowOffset:=ShadowOffset, DefaultShadowOpacity := DefaultTextShadowOpacity}
    R_Row2_AgedSapphires : text_block = text_block{DefaultTextColor:=AgedSapphiresTextColor, DefaultShadowOffset:=ShadowOffset, DefaultShadowOpacity := DefaultTextShadowOpacity}
    R_Row3_Fame : text_block = text_block{DefaultTextColor:=FameTextColor, DefaultShadowOffset:=ShadowOffset, DefaultShadowOpacity := DefaultTextShadowOpacity}
    R_Row4_Water : text_block = text_block{DefaultTextColor:=WaterTextColor, DefaultShadowOffset:=ShadowOffset, DefaultShadowOpacity := DefaultTextShadowOpacity}
    R_Row5_Flour : text_block = text_block{DefaultTextColor:=FlourTextColor, DefaultShadowOffset:=ShadowOffset, DefaultShadowOpacity := DefaultTextShadowOpacity}
    R_Row6_Cheese : text_block = text_block{DefaultTextColor:=CheeseTextColor, DefaultShadowOffset:=ShadowOffset, DefaultShadowOpacity := DefaultTextShadowOpacity}
    R_Row7_Pimentos : text_block = text_block{DefaultTextColor:=PimentoTextColor, DefaultShadowOffset:=ShadowOffset, DefaultShadowOpacity := DefaultTextShadowOpacity}
    R_Row8_Mutant : text_block = text_block{DefaultTextColor:=MutantMaterialTextColor, DefaultShadowOffset:=ShadowOffset, DefaultShadowOpacity := DefaultTextShadowOpacity}
    R_Row9_Coal : text_block = text_block{DefaultTextColor:=CoalTextColor, DefaultShadowOffset:=ShadowOffset, DefaultShadowOpacity := DefaultTextShadowOpacity}

    var MaybeCanvas : ?canvas = false
    var UIIsDisplayed : logic = false

    var TimeElapsed : int = 0
    
    #Close
    CloseButton : button_loud = button_loud{}


    InitUI() : void =
        UpdateUI()
        Canvas := MakeCanvas()
        set MaybeCanvas = option{Canvas}

    SetAgent(InAgent : agent) : void =
        set MaybeAgent = option{InAgent}

    UpdateUI() : void =

        R_Row1_Sapphires.SetText(BasicMessage("60.0"))
        R_Row2_AgedSapphires.SetText(BasicMessage("60.0"))
        R_Row3_Fame.SetText(BasicMessage("60.0"))
        R_Row4_Water.SetText(BasicMessage("60.0"))
        R_Row5_Flour.SetText(BasicMessage("60.0"))
        R_Row6_Cheese.SetText(BasicMessage("60.0"))
        R_Row7_Pimentos.SetText(BasicMessage("60.0"))
        R_Row8_Mutant.SetText(BasicMessage("60.0"))
        R_Row9_Coal.SetText(BasicMessage("60.0"))

    Remove_Canvas(Message : widget_message) : void =
        set UIIsDisplayed = false
        #Print("Trying to remove UI")
        if:
            Agent := MaybeAgent?
            Player:=player[Agent]
            PlayerUI:=GetPlayerUI[Player]
            Canvas := MaybeCanvas?
        then:
            #Print("Removing UI")
            PlayerUI.RemoveWidget(Canvas)

    Remove_CanvasWithoutMessage() : void =
        set UIIsDisplayed = false
        #Print("Trying to remove UI")
        if:
            Agent := MaybeAgent?
            Player:=player[Agent]
            PlayerUI:=GetPlayerUI[Player]
            Canvas := MaybeCanvas?
        then:
            #Print("Removing UI")
            PlayerUI.RemoveWidget(Canvas)

            
    Display_Canvas() : void =
        set TimeElapsed = 0
        set UIIsDisplayed = true
        spawn: 
            StartTimeElapsed()
        if:
            Agent := MaybeAgent?
            PlayerUI:=GetPlayerUI[player[Agent]]
            Canvas := MaybeCanvas?
        then:
            #Print("Assigning created UI before")
            PlayerUI.AddWidget(Canvas,player_ui_slot{InputMode := ui_input_mode.All, ZOrder:=1})  
        #UpdateUI()

    StartTimeElapsed()<suspends> : void =
        loop:
            #Print("IN THE LOOP TimeElapsed: {TimeElapsed}")
            if(TimeElapsed > 20):
                #Print("Closing the UI because 20 seconds has pased")
                Remove_CanvasWithoutMessage()
                break
            if(UIIsDisplayed = false):
                #Print("Breaking Time Elapsed loop")
                break
            Sleep(1.0)
            set TimeElapsed = TimeElapsed + 1


    MakeCanvas(): canvas=
        Canvas := canvas:
            Slots:=array:
                #Main overlay
                canvas_slot:
                    ZOrder:=0
                    Anchors:=anchors{Minimum:=vector2{X:=0.0,Y:=0.5},Maximum:=vector2{X:=0.0,Y:=0.5}}
                    Offsets:=margin{Left:=50.0,Top:=0.0,Right:=0.0,Bottom:=0.0} #PositionX, PositionY, SizeX, SizeY
                    Alignment:=vector2{X:=0.0,Y:=0.5}
                    SizeToContent:=true
                    Widget:=overlay:
                        Slots:=array:
                            overlay_slot:
                                Widget:=texture_block{DefaultDesiredSize:=vector2{X:=ResourcersX,Y:=ResourcersY},DefaultImage:=Images.EmptyTexture}
                            overlay_slot:
                                HorizontalAlignment:=horizontal_alignment.Fill
                                VerticalAlignment:=vertical_alignment.Fill
                                Widget:=stack_box:
                                    Orientation:=orientation.Vertical
                                        Slots:=array:
                                            stack_box_slot: #Line 1
                                                Padding:=margin{Left:=ResourcersRowLeftPadding,Top:=ResourcersRowTopPadding,Right:=ResourcersRowRightPadding,Bottom:=ResourcersRowBottomPadding}
                                                HorizontalAlignment:=horizontal_alignment.Fill
                                                VerticalAlignment:=vertical_alignment.Fill
                                                Widget:=overlay:
                                                    Slots:=array:
                                                        overlay_slot: 
                                                            HorizontalAlignment:=horizontal_alignment.Fill
                                                            VerticalAlignment:=vertical_alignment.Fill
                                                            Widget:=Background_TextureBlock
                                                        overlay_slot: 
                                                            Padding:=TextPadding
                                                            HorizontalAlignment:=horizontal_alignment.Left
                                                            VerticalAlignment:=vertical_alignment.Center
                                                            Widget:=R_Row1_Sapphires
                                                        overlay_slot: 
                                                            Padding:=IconPadding
                                                            HorizontalAlignment:=horizontal_alignment.Left
                                                            VerticalAlignment:=vertical_alignment.Center 
                                                            Widget:=Sapphires_TextureBlock
                                            stack_box_slot: #Line 2
                                                Padding:=margin{Left:=ResourcersRowLeftPadding,Top:=ResourcersRowTopPadding,Right:=ResourcersRowRightPadding,Bottom:=ResourcersRowBottomPadding}
                                                HorizontalAlignment:=horizontal_alignment.Fill
                                                VerticalAlignment:=vertical_alignment.Fill
                                                Widget:=overlay:
                                                    Slots:=array:
                                                        overlay_slot: 
                                                            HorizontalAlignment:=horizontal_alignment.Fill
                                                            VerticalAlignment:=vertical_alignment.Fill
                                                            Widget:=Background_TextureBlock
                                                        overlay_slot: 
                                                            Padding:=TextPadding
                                                            HorizontalAlignment:=horizontal_alignment.Left
                                                            VerticalAlignment:=vertical_alignment.Center
                                                            Widget:=R_Row2_AgedSapphires
                                                        overlay_slot: 
                                                            Padding:=IconPadding
                                                            HorizontalAlignment:=horizontal_alignment.Left
                                                            VerticalAlignment:=vertical_alignment.Center 
                                                            Widget:=AgedSapphires_TextureBlock
                                            stack_box_slot: #Line 3
                                                Padding:=margin{Left:=ResourcersRowLeftPadding,Top:=ResourcersRowTopPadding,Right:=ResourcersRowRightPadding,Bottom:=ResourcersRowBottomPadding}
                                                HorizontalAlignment:=horizontal_alignment.Fill
                                                VerticalAlignment:=vertical_alignment.Fill
                                                Widget:=overlay:
                                                    Slots:=array:
                                                        overlay_slot: 
                                                            HorizontalAlignment:=horizontal_alignment.Fill
                                                            VerticalAlignment:=vertical_alignment.Fill
                                                            Widget:=Background_TextureBlock
                                                        overlay_slot: 
                                                            Padding:=TextPadding
                                                            HorizontalAlignment:=horizontal_alignment.Left
                                                            VerticalAlignment:=vertical_alignment.Center
                                                            Widget:=R_Row3_Fame
                                                        overlay_slot: 
                                                            Padding:=IconPadding
                                                            HorizontalAlignment:=horizontal_alignment.Left
                                                            VerticalAlignment:=vertical_alignment.Center 
                                                            Widget:=Fame_TextureBlock
                                            stack_box_slot: #Line 4
                                                Padding:=margin{Left:=ResourcersRowLeftPadding,Top:=ResourcersRowTopPadding,Right:=ResourcersRowRightPadding,Bottom:=ResourcersRowBottomPadding}
                                                HorizontalAlignment:=horizontal_alignment.Fill
                                                VerticalAlignment:=vertical_alignment.Fill
                                                Widget:=overlay:
                                                    Slots:=array:
                                                        overlay_slot: 
                                                            HorizontalAlignment:=horizontal_alignment.Fill
                                                            VerticalAlignment:=vertical_alignment.Fill
                                                            Widget:=Background_TextureBlock
                                                        overlay_slot: 
                                                            Padding:=TextPadding
                                                            HorizontalAlignment:=horizontal_alignment.Left
                                                            VerticalAlignment:=vertical_alignment.Center
                                                            Widget:=R_Row4_Water
                                                        overlay_slot: 
                                                            Padding:=IconPadding
                                                            HorizontalAlignment:=horizontal_alignment.Left
                                                            VerticalAlignment:=vertical_alignment.Center 
                                                            Widget:=Water_TextureBlock
                                            stack_box_slot: #Line 5
                                                Padding:=margin{Left:=ResourcersRowLeftPadding,Top:=ResourcersRowTopPadding,Right:=ResourcersRowRightPadding,Bottom:=ResourcersRowBottomPadding}
                                                HorizontalAlignment:=horizontal_alignment.Fill
                                                VerticalAlignment:=vertical_alignment.Fill
                                                Widget:=overlay:
                                                    Slots:=array:
                                                        overlay_slot: 
                                                            HorizontalAlignment:=horizontal_alignment.Fill
                                                            VerticalAlignment:=vertical_alignment.Fill
                                                            Widget:=Background_TextureBlock
                                                        overlay_slot: 
                                                            Padding:=TextPadding
                                                            HorizontalAlignment:=horizontal_alignment.Left
                                                            VerticalAlignment:=vertical_alignment.Center
                                                            Widget:=R_Row5_Flour
                                                        overlay_slot: 
                                                            Padding:=IconPadding
                                                            HorizontalAlignment:=horizontal_alignment.Left
                                                            VerticalAlignment:=vertical_alignment.Center 
                                                            Widget:=Flour_TextureBlock
                                            stack_box_slot: #Line 6
                                                Padding:=margin{Left:=ResourcersRowLeftPadding,Top:=ResourcersRowTopPadding,Right:=ResourcersRowRightPadding,Bottom:=ResourcersRowBottomPadding}
                                                HorizontalAlignment:=horizontal_alignment.Fill
                                                VerticalAlignment:=vertical_alignment.Fill
                                                Widget:=overlay:
                                                    Slots:=array:
                                                        overlay_slot: 
                                                            HorizontalAlignment:=horizontal_alignment.Fill
                                                            VerticalAlignment:=vertical_alignment.Fill
                                                            Widget:=Background_TextureBlock
                                                        overlay_slot: 
                                                            Padding:=TextPadding
                                                            HorizontalAlignment:=horizontal_alignment.Left
                                                            VerticalAlignment:=vertical_alignment.Center
                                                            Widget:=R_Row6_Cheese
                                                        overlay_slot: 
                                                            Padding:=IconPadding
                                                            HorizontalAlignment:=horizontal_alignment.Left
                                                            VerticalAlignment:=vertical_alignment.Center 
                                                            Widget:=Cheese_TextureBlock
                                            stack_box_slot: #Line 7
                                                Padding:=margin{Left:=ResourcersRowLeftPadding,Top:=ResourcersRowTopPadding,Right:=ResourcersRowRightPadding,Bottom:=ResourcersRowBottomPadding}
                                                HorizontalAlignment:=horizontal_alignment.Fill
                                                VerticalAlignment:=vertical_alignment.Fill
                                                Widget:=overlay:
                                                    Slots:=array:
                                                        overlay_slot: 
                                                            HorizontalAlignment:=horizontal_alignment.Fill
                                                            VerticalAlignment:=vertical_alignment.Fill
                                                            Widget:=Background_TextureBlock
                                                        overlay_slot: 
                                                            Padding:=TextPadding
                                                            HorizontalAlignment:=horizontal_alignment.Left
                                                            VerticalAlignment:=vertical_alignment.Center
                                                            Widget:=R_Row7_Pimentos
                                                        overlay_slot: 
                                                            Padding:=IconPadding
                                                            HorizontalAlignment:=horizontal_alignment.Left
                                                            VerticalAlignment:=vertical_alignment.Center 
                                                            Widget:=Pimentos_TextureBlock
                                            stack_box_slot: #Line 8
                                                Padding:=margin{Left:=ResourcersRowLeftPadding,Top:=ResourcersRowTopPadding,Right:=ResourcersRowRightPadding,Bottom:=ResourcersRowBottomPadding}
                                                HorizontalAlignment:=horizontal_alignment.Fill
                                                VerticalAlignment:=vertical_alignment.Fill
                                                Widget:=overlay:
                                                    Slots:=array:
                                                        overlay_slot: 
                                                            HorizontalAlignment:=horizontal_alignment.Fill
                                                            VerticalAlignment:=vertical_alignment.Fill
                                                            Widget:=Background_TextureBlock
                                                        overlay_slot: 
                                                            Padding:=TextPadding
                                                            HorizontalAlignment:=horizontal_alignment.Left
                                                            VerticalAlignment:=vertical_alignment.Center
                                                            Widget:=R_Row8_Mutant
                                                        overlay_slot: 
                                                            Padding:=IconPadding
                                                            HorizontalAlignment:=horizontal_alignment.Left
                                                            VerticalAlignment:=vertical_alignment.Center 
                                                            Widget:=Mutant_TextureBlock
                                                                                       
                    
MakeResourersUI<constructor>(
    InBakeryMachinesManager : bakery_machines_manager)<transacts> := resourcers_ui :
        BakeryMachinesManager := InBakeryMachinesManager
                                                                                                               