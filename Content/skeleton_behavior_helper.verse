# See https://dev.epicgames.com/documentation/en-us/uefn/create-your-own-device-in-verse for how to create a verse device.

using { /Fortnite.com/Devices }
using { /UnrealEngine.com/Temporary/SpatialMath }
using { /Fortnite.com/Characters }
using { /Verse.org/Simulation }

# A Verse-authored creative device that can be placed in a level
skeleton_behavior_helper := class(creative_device):
    
    FindNearestTargetFromPos(FortChar : fort_character)< decides>< transacts> : fort_character = {
        var MaybeTarget : ?fort_character = false

        #Change this range how you would like
        #This will be the range that the device will search for the nearest target
        var CheckRange : float = 5000.0

        for (Player : GetPlayspace().GetPlayers(), PlayerFortChar := Player.GetFortCharacter[]) {
            DistanceDifference := Distance(FortChar.GetTransform().Translation, PlayerFortChar.GetTransform().Translation)

            if (DistanceDifference < CheckRange and not PlayerFortChar = FortChar) {
                set MaybeTarget = option{PlayerFortChar}
                set CheckRange = DistanceDifference 
            }
        }

        return MaybeTarget?
    }

    <#
    DetectBosses() : void =
        ##Print("Detecting bosses")
        TaggedActors := GetCreativeObjectsWithTag(mutant_hardbread{})
        for(Actor : TaggedActors):
            #Print("Detected boss")
            if(Prop:=creative_prop[Actor]):
                #Print("Changing material")
                Prop.SetMaterial(Mat_Green_FollowMe)
                Prop.SetMaterial(Mat_Green_FollowMe, ?Index:=0)#>

    # Runs when the device is started in a running game
    OnBegin<override>()<suspends>:void=
        loop:
            Sleep(5.0)
            #DetectBosses()
            