using { /Fortnite.com/Devices }
using { /Verse.org/Simulation }
using { /Verse.org/Simulation/Tags }

# Derive from the `tag` class in the Verse.org/Simulation/Tags module to create a new gameplay tag.
B_prop := class(tag){}