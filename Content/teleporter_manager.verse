
using { /Fortnite.com/Devices }
using { /Verse.org/Simulation }
using { /UnrealEngine.com/Temporary/Diagnostics }
using { /Verse.org/Assets }
using { /Fortnite.com/Characters }
using { /UnrealEngine.com/Temporary/SpatialMath }

# See https://dev.epicgames.com/documentation/en-us/uefn/create-your-own-device-in-verse for how to create a verse device.

# A Verse-authored creative device that can be placed in a level
teleporter_manager := class(creative_device):

    var MaybeGameManager : ?game_manager = false

    @editable
    var CostPerZone : float = 5.0

    #All teleporter
    #HexaZone
    @editable
    HexaGameTeleporterProp : creative_prop = creative_prop{}
    @editable
    HexaGameTeleporterButton : button_device = button_device{}
    #LobbyZone
    @editable
    LobbyTeleporterProp : creative_prop = creative_prop{}
    @editable
    LobbyTeleporterButton : button_device = button_device{}
    #All areas
    @editable
    AllTeleporterAreaProps : []creative_prop = array{}
    @editable
    AllTeleporterAreaButtons : []button_device = array{}
    var AllTeleporterCosts : []float = array{}
    # Runs when the device is started in a running game
    OnBegin<override>()<suspends>:void=
        SetCosts()

    SetGameManager(GameManager : game_manager) : void =
        set MaybeGameManager = option{GameManager}
        InitSubscribers()

    InitSubscribers() : void =

        LobbyTeleporterButton.InteractedWithEvent.Subscribe(handler_random{Index:=TELEPORTER_FROM_LOBBY_AREA, MaybeGameManager := MaybeGameManager}.HandlerOpenTeleporterMenu)   
        HexaGameTeleporterButton.InteractedWithEvent.Subscribe(handler_random{Index:=TELEPORTER_FROM_HEXAGAME_AREA, MaybeGameManager := MaybeGameManager}.HandlerOpenTeleporterMenu)
        for(Index:=0..AllTeleporterAreaButtons.Length-1, Button:=AllTeleporterAreaButtons[Index]):
            Button.InteractedWithEvent.Subscribe(handler_random{MaybeTeleporterManager:=option{Self}, Index:=Index}.HandlerTeleporterButton)

    TryToTeleportToBakery(Index : int, Agent : agent) : void =
        if:
            GameManager := MaybeGameManager?
            Cost:=AllTeleporterCosts[Index]
            Player := player[Agent]
            Player.IsActive[]
            ValidBakery:=GameManager.GetAgentBakery_Transacted(Agent)?
        then:
            if(Cost<=ValidBakery.PlayerDataCache.Inventory_Coal):
                ValidBakery.CheckMusicPlay(BAKERSMUSIC_SOUND)
                UpdatedPlayerData : player_data_table = GameManager.PlayersDataManager.WithdrawResource(ValidBakery.PlayerDataCache,Cost,COAL)
                MaybeBakery:=GameManager.GetAgentBakery(Agent)
                ValidBakery.SetNewPlayerDataCache(UpdatedPlayerData)
                #GameManager.PlayersDataManager.SaveAllNewData(Player, PlayerData,UpdatedPlayerData, ValidBakery)
                ValidBakery.RefreshInventoryUI(Agent)
                if(Character:=Agent.GetFortCharacter[], Character.IsActive[], ObjetiveTransform :=ValidBakery.Prop_PersonalTeleporter.GetTransform()):
                    Character.Heal(99999.0)  
                    TeleportPlayerUsingTeleporter(Agent,ObjetiveTransform)
                    ValidBakery.CheckQuestProgress(1.0,USETELEPORTER_TUTORIAL_TYPE)
                    ValidBakery.CheckQuestProgress(Cost,USEEXPENSIVETELEPORTER_TUTORIAL_TYPE)
                    ValidBakery.DisplayResourcersGather(array{-Cost},array{COAL},false)
                    #GameManager.VFXManager.PlayerPays_AudioPlayer.Play(Agent)
            else:
                GameManager.VFXManager.ErrorSound_AudioPlayer.Play(Agent)
                GameManager.DisplayCharacterMessage(Agent, W_CHARACTER, "You need {GetCharMagnitudeLetter(Cost-ValidBakery.PlayerDataCache.Inventory_Coal)} more Coal!")

    TryToFreeTeleportToBakery(Agent : agent) : void =
        if:
            GameManager := MaybeGameManager?
            Player := player[Agent]
            Player.IsActive[]
            ValidBakery:=GameManager.GetAgentBakery_Transacted(Agent)?
        then:
            ValidBakery.CheckMusicPlay(BAKERSMUSIC_SOUND)
            Log("Trying to teleport Agent to his Bakery",LOG_LEVEL_DEBUG2)
            MaybeBakery:=GameManager.GetAgentBakery(Agent)
            if(Character:=Agent.GetFortCharacter[], Character.IsActive[], ObjetiveTransform :=ValidBakery.Prop_PersonalTeleporter.GetTransform()):
                Character.Heal(99999.0)  
                TeleportPlayerUsingTeleporter(Agent,ObjetiveTransform)
                ValidBakery.CheckQuestProgress(1.0,USETELEPORTER_TUTORIAL_TYPE)
            else:
                PrintFatalError("On TryToFreeTeleportToBakery ObjetiveTransform to teleport not found!")


    TryToTeleportPaidToArea(ArenaNum : int, Agent : agent) : void =
        if:
            GameManager := MaybeGameManager?
            Player := player[Agent]
            Player.IsActive[]
        then:
            MaybeBakery:=GameManager.GetAgentBakery(Agent)
            if(Character:=Agent.GetFortCharacter[], Character.IsActive[], ObjetiveTransform :=AllTeleporterAreaProps[ArenaNum-1].GetTransform()):
                Character.Heal(99999.0)  
                TeleportPlayerUsingTeleporter(Agent,ObjetiveTransform)

    TryToTeleportToHexaGame(Agent : agent) : void =
        if:
            GameManager := MaybeGameManager?
            Player := player[Agent]
            Player.IsActive[]
        then:
            MaybeBakery:=GameManager.GetAgentBakery(Agent)
            if(Character:=Agent.GetFortCharacter[], Character.IsActive[]):
                ObjetiveTransform :=HexaGameTeleporterProp.GetTransform()
                Character.Heal(99999.0)  
                TeleportPlayerUsingTeleporter(Agent,ObjetiveTransform)

    TryToTeleportToLobby(Agent : agent) : void =
        if:
            GameManager := MaybeGameManager?
            Player := player[Agent]
            Player.IsActive[]
        then:
            MaybeBakery:=GameManager.GetAgentBakery(Agent)
            if(Character:=Agent.GetFortCharacter[], Character.IsActive[]):
                ObjetiveTransform :=LobbyTeleporterProp.GetTransform()
                Character.Heal(99999.0)  
                TeleportPlayerUsingTeleporter(Agent,ObjetiveTransform)

    TeleportPlayerUsingTeleporter(Agent : agent, TeleporterTransform : transform) : void =
        if:
            GameManager := MaybeGameManager?
            Player := player[Agent]
            Player.IsActive[]
            Character:=Agent.GetFortCharacter[]
            Character.IsActive[]
        then:
            NewX := TeleporterTransform.Translation.X
            NewY := TeleporterTransform.Translation.Y
            NewZ := TeleporterTransform.Translation.Z+32.0

            NewTranslation := vector3{X:=NewX,Y:=NewY,Z:=NewZ}
            NewTransform := transform{Translation:=NewTranslation,Rotation:=TeleporterTransform.Rotation,Scale:=TeleporterTransform.Scale}

            if(Character.TeleportTo[NewTranslation, TeleporterTransform.Rotation]):
                GameManager.VFXManager.SpawnTeleportEffect(Agent,GameManager.GetAgentIndex(Agent),NewTransform)
            else:
                ResultTuple:=TeleportFailedAttempt(option{Character},false,false,NewTransform,256.0)
                if(not ResultTuple(0)?):
                    Character.Damage(5000.0)
                else:
                    GameManager.VFXManager.SpawnTeleportEffect(Agent,GameManager.GetAgentIndex(Agent),ResultTuple(1))


    SetCosts<private>() : void =
        for(Index:=0..AllTeleporterAreaButtons.Length-1):
            if:
                TeleporterProp:=AllTeleporterAreaProps[Index]
                Button:=AllTeleporterAreaButtons[Index]
            then:
                Cost:=Index*CostPerZone
                SetMaterialCost(TeleporterProp,Button,Cost)
                set AllTeleporterCosts = AllTeleporterCosts + array{Cost}
      
    SetMaterialCost<private>(CreativeProp : creative_prop, Button : button_device, AmountCost : float) : void =
        if:
            Decenas := Floor[AmountCost/10.0]
            Unidades := Floor[AmountCost-(Decenas*10.0)]
            IntAmount := Floor[AmountCost]
        then:
            if(DecenasMaterial:=AllMaterialnumbers[Decenas], UnidadesMaterial:=AllMaterialnumbers[Unidades]):
                CreativeProp.SetMaterial(Materials.Mat_Icon_Coal,?Index:=3)
                CreativeProp.SetMaterial(DecenasMaterial,?Index:=4)
                CreativeProp.SetMaterial(UnidadesMaterial,?Index:=5)
                Button.SetInteractionText(BasicMessage("Teleport to the factory. Cost: x{IntAmount} Coal"))
       