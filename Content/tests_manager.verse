using { /Fortnite.com/Devices }
using { /Verse.org/Simulation }
using { /UnrealEngine.com/Temporary/Diagnostics }
using { /Verse.org/Random }
using { /Fortnite.com/Characters }
using { /UnrealEngine.com/Temporary/SpatialMath }
using { /Fortnite.com/Devices/CreativeAnimation }
using { /Fortnite.com/Devices/CreativeAnimation/InterpolationTypes }
using { /UnrealEngine.com/Temporary/UI }
using { /Fortnite.com/UI }
using { /Verse.org/Colors/NamedColors }
using { /Verse.org/Colors }

# See https://dev.epicgames.com/documentation/en-us/uefn/create-your-own-device-in-verse for how to create a verse device.

# A Verse-authored creative device that can be placed in a level
tests_manager := class(creative_device):

    @editable
    AllPropManipulators : []prop_manipulator_device = array{}

    #PickaxeCalculator : weapon_calculator = pickaxe_calculator{}

    var TotalGameTime : float = 0.0

    @editable
    PetBlenderProp : creative_prop = creative_prop{}

    @editable
    Camera_PetCreation : gameplay_camera_fixed_point_device = gameplay_camera_fixed_point_device{}

    

    @editable
    GeneralMobsManager : general_mobs_manager = general_mobs_manager{}

    @editable
    ClassHUD_NoHealth : class_and_team_selector_device = class_and_team_selector_device{}
    @editable
    ClassHUD_Standard : class_and_team_selector_device = class_and_team_selector_device{}

    @editable
    ButtonUpgradeWeapons : button_device = button_device{}

    #@editable
    #AllHexaManager : []hexa_minigame_manager = array{}
    #var AllPetsManager : [agent]pets_manager = map{}

    @editable
    CollisionTestPetProp : creative_prop = creative_prop{}

    @editable
    PetMovementRefresh : float = 1.0

    #Prop
    @editable
    AllPetPropAsset : []creative_prop_asset = array{}
    #Attack effect
    @editable
    ExplosionEffect : creative_prop_asset = DefaultCreativePropAsset
    @editable
    MeleeAttackEffect : creative_prop_asset = DefaultCreativePropAsset
    @editable
    DefendEffect : creative_prop_asset = DefaultCreativePropAsset
    @editable
    RangeAttackEffect : creative_prop_asset = DefaultCreativePropAsset

    @editable
    HealAndShieldEffect : creative_prop_asset = DefaultCreativePropAsset

    @editable
    UITest : hud_message_device = hud_message_device{}

    PlayersDataManager : player_data_manager = player_data_manager{}

    @editable
    ShowColorsButton : button_device = button_device{}

    @editable
    CompleteTutorialButton : button_device = button_device{}
    @editable
    AddProgressToOrder_Button : button_device = button_device{}
    @editable
    ButtonTest1 : button_device = button_device{}
    @editable
    ButtonTest2 : button_device = button_device{}
    @editable
    ButtonTest3 : button_device = button_device{}
    @editable
    ButtonTest4 : button_device = button_device{}
    @editable
    ButtonTest5 : button_device = button_device{}
    @editable
    ButtonGrantSuperLaser : button_device = button_device{}
    @editable
    ButtonRemoveSuperLaser : button_device = button_device{}
    @editable
    ButtonGrantSuperSword : button_device = button_device{}
    @editable
    ButtonRemoveSuperSword : button_device = button_device{}

    @editable
    HUDMessage : hud_message_device = hud_message_device{}
    

    @editable
    AllPlayerSpawners : []player_spawner_device = array{}

    var TotalWater : float = 0.0

    #var MaybeEntityHealthUI : ?entity_health_bar_ui = false
    var MaybeActionInfoUI : ?action_info_ui = false
    var MaybeMainInventoryUI : ?main_inventory_ui = false
    #var MaybeBakerInfoUI : ?pcsb_info_ui = false

    var AllMainInventoryUI : [agent]main_inventory_ui = map{}
    var AllTokensUI : [agent]tokens_ui = map{}
    var AllActionInfoUI : [agent]action_info_ui = map{}

    var IsGivenPets : [agent]logic = map{}

    var MaybePetUpgradesUI : ?pets_upgrades_ui = false

    var MaybeFastTravelUI : ?fast_travel_ui = false

    var HardBreadHealth : float = 1000000.0
    var HardBreadMaxHealth : float = 1000000.0

    #Color canvas
    var MaybeColorCanvas : ?canvas = false
    var PageColor : int  = 0

    var ArrayColors : []color = array{AliceBlue, AntiqueWhite, Aqua, Aquamarine, Azure, Beige, Bisque, Black, BlanchedAlmond, Blue, BlueViolet, Brown, Burlywood, CadetBlue, Chartreuse, Chocolate, Coral, CornflowerBlue, Cornsilk, Crimson, Cyan, DarkBlue, DarkCyan, DarkGoldenrod, DarkGray, DarkGreen, DarkGrey, DarkKhaki, DarkMagenta, DarkOliveGreen, DarkOrange, DarkOrchid, DarkRed, DarkSalmon, DarkSeaGreen, DarkSlateBlue, DarkSlateGray, DarkSlateGrey, DarkTurquoise, DarkViolet, DeepPink, DeepSkyBlue, DimGray, DimGrey, DodgerBlue, Firebrick, FloralWhite, ForestGreen, Fuchsia, Gainsboro, GhostWhite, Gold, Goldenrod, Gray, Green, GreenYellow, Grey, Honeydew, Hotpink, IndianRed, Indigo, Ivory, Khaki, Lavender, LavenderBlush, LawnGreen, LemonChiffon, LightBlue, LightCoral, LightCyan, LightGoldenrodYellow, LightGray, LightGreen, LightGrey, LightPink, LightSalmon, LightSeaGreen, LightSkyBlue, LightSlateGray, LightSlateGrey, LightSteelBlue, LightYellow, Lime, LimeGreen, Linen, Magenta, Maroon, MediumAquamarine, MediumBlue, MediumOrchid, MediumPurple, MediumSeaGreen, MediumSlateBlue, MediumSpringGreen, MediumTurquoise, MediumVioletRed, MidnightBlue, MintCream, MistyRose, Moccasin, NavajoWhite, Navy, OldLace, Olive, OliveDrab, Orange, OrangeRed, Orchid, PaleGoldenrod, PaleGreen, PaleTurquoise, PaleVioletred, PapayaWhip, PeachPuff, Peru, Pink, Plum, PowderBlue, Purple, Red, RosyBrown, RoyalBlue, SaddleBrown, Salmon, SandyBrown, SeaGreen, SeaShell, Sienna, Silver, SkyBlue, SlateBlue, SlateGray, SlateGrey, Snow, SpringGreen, SteelBlue, Teal, Thistle, Tomato, Turquoise, Violet, Wheat, White, WhiteSmoke, Yellow, YellowGreen}
    var ArrayColorNames : []string = array{"AliceBlue", "AntiqueWhite", "Aqua", "Aquamarine", "Azure", "Beige", "Bisque", "Black", "BlanchedAlmond", "Blue", "BlueViolet", "Brown", "Burlywood", "CadetBlue", "Chartreuse", "Chocolate", "Coral", "CornflowerBlue", "Cornsilk", "Crimson", "Cyan", "DarkBlue", "DarkCyan", "DarkGoldenrod", "DarkGray", "DarkGreen", "DarkGrey", "DarkKhaki", "DarkMagenta", "DarkOliveGreen", "DarkOrange", "DarkOrchid", "DarkRed", "DarkSalmon", "DarkSeaGreen", "DarkSlateBlue", "DarkSlateGray", "DarkSlateGrey", "DarkTurquoise", "DarkViolet", "DeepPink", "DeepSkyBlue", "DimGray", "DimGrey", "DodgerBlue", "Firebrick", "FloralWhite", "ForestGreen", "Fuchsia", "Gainsboro", "GhostWhite", "Gold", "Goldenrod", "Gray", "Green", "GreenYellow", "Grey", "Honeydew", "Hotpink", "IndianRed", "Indigo", "Ivory", "Khaki", "Lavender", "LavenderBlush", "LawnGreen", "LemonChiffon", "LightBlue", "LightCoral", "LightCyan", "LightGoldenrodYellow", "LightGray", "LightGreen", "LightGrey", "LightPink", "LightSalmon", "LightSeaGreen", "LightSkyBlue", "LightSlateGray", "LightSlateGrey", "LightSteelBlue", "LightYellow", "Lime", "LimeGreen", "Linen", "Magenta", "Maroon", "MediumAquamarine", "MediumBlue", "MediumOrchid", "MediumPurple", "MediumSeaGreen", "MediumSlateBlue", "MediumSpringGreen", "MediumTurquoise", "MediumVioletRed", "MidnightBlue", "MintCream", "MistyRose", "Moccasin", "NavajoWhite", "Navy", "OldLace", "Olive", "OliveDrab", "Orange", "OrangeRed", "Orchid", "PaleGoldenrod", "PaleGreen", "PaleTurquoise", "PaleVioletred", "PapayaWhip", "PeachPuff", "Peru", "Pink", "Plum", "PowderBlue", "Purple", "Red", "RosyBrown", "RoyalBlue", "SaddleBrown", "Salmon", "SandyBrown", "SeaGreen", "SeaShell", "Sienna", "Silver", "SkyBlue", "SlateBlue", "SlateGray", "SlateGrey", "Snow", "SpringGreen", "SteelBlue", "Teal", "Thistle", "Tomato", "Turquoise", "Violet", "Wheat", "White", "WhiteSmoke", "Yellow", "YellowGreen"}

    ChangeColorButton : button_loud = button_loud{}
    CloseColorButton : button_loud = button_loud{}

    TextBlock0  : text_block = text_block{DefaultTextColor := White, DefaultShadowOffset:=ShadowOffset, DefaultShadowOpacity := DefaultTextShadowOpacity}
    TextBlock1  : text_block = text_block{DefaultTextColor := AntiqueWhite, DefaultShadowOffset:=ShadowOffset, DefaultShadowOpacity := DefaultTextShadowOpacity}
    TextBlock2  : text_block = text_block{DefaultTextColor := Aqua, DefaultShadowOffset:=ShadowOffset, DefaultShadowOpacity := DefaultTextShadowOpacity}
    TextBlock3  : text_block = text_block{DefaultTextColor := Aquamarine, DefaultShadowOffset:=ShadowOffset, DefaultShadowOpacity := DefaultTextShadowOpacity}
    TextBlock4  : text_block = text_block{DefaultTextColor := Azure, DefaultShadowOffset:=ShadowOffset, DefaultShadowOpacity := DefaultTextShadowOpacity}
    TextBlock5  : text_block = text_block{DefaultTextColor := Beige, DefaultShadowOffset:=ShadowOffset, DefaultShadowOpacity := DefaultTextShadowOpacity}
    TextBlock6  : text_block = text_block{DefaultTextColor := Bisque, DefaultShadowOffset:=ShadowOffset, DefaultShadowOpacity := DefaultTextShadowOpacity}
    TextBlock7  : text_block = text_block{DefaultTextColor := Black, DefaultShadowOffset:=ShadowOffset, DefaultShadowOpacity := DefaultTextShadowOpacity}
    TextBlock8  : text_block = text_block{DefaultTextColor := BlanchedAlmond, DefaultShadowOffset:=ShadowOffset, DefaultShadowOpacity := DefaultTextShadowOpacity}
    TextBlock9  : text_block = text_block{DefaultTextColor := Blue, DefaultShadowOffset:=ShadowOffset, DefaultShadowOpacity := DefaultTextShadowOpacity}
    TextBlock10 : text_block = text_block{DefaultTextColor := BlueViolet, DefaultShadowOffset:=ShadowOffset, DefaultShadowOpacity := DefaultTextShadowOpacity}
    TextBlock11 : text_block = text_block{DefaultTextColor := Brown, DefaultShadowOffset:=ShadowOffset, DefaultShadowOpacity := DefaultTextShadowOpacity}
    TextBlock12 : text_block = text_block{DefaultTextColor := Burlywood, DefaultShadowOffset:=ShadowOffset, DefaultShadowOpacity := DefaultTextShadowOpacity}
    TextBlock13 : text_block = text_block{DefaultTextColor := CadetBlue, DefaultShadowOffset:=ShadowOffset, DefaultShadowOpacity := DefaultTextShadowOpacity}
    TextBlock14 : text_block = text_block{DefaultTextColor := Chartreuse, DefaultShadowOffset:=ShadowOffset, DefaultShadowOpacity := DefaultTextShadowOpacity}
    TextBlock15 : text_block = text_block{DefaultTextColor := Chocolate, DefaultShadowOffset:=ShadowOffset, DefaultShadowOpacity := DefaultTextShadowOpacity}
    TextBlock16 : text_block = text_block{DefaultTextColor := Coral, DefaultShadowOffset:=ShadowOffset, DefaultShadowOpacity := DefaultTextShadowOpacity}
    TextBlock17 : text_block = text_block{DefaultTextColor := CornflowerBlue, DefaultShadowOffset:=ShadowOffset, DefaultShadowOpacity := DefaultTextShadowOpacity}
    TextBlock18 : text_block = text_block{DefaultTextColor := Cornsilk, DefaultShadowOffset:=ShadowOffset, DefaultShadowOpacity := DefaultTextShadowOpacity}
    TextBlock19 : text_block = text_block{DefaultTextColor := Crimson, DefaultShadowOffset:=ShadowOffset, DefaultShadowOpacity := DefaultTextShadowOpacity}
    TextBlock20 : text_block = text_block{DefaultTextColor := Cyan, DefaultShadowOffset:=ShadowOffset, DefaultShadowOpacity := DefaultTextShadowOpacity}
    TextBlock21 : text_block = text_block{DefaultTextColor := DarkBlue, DefaultShadowOffset:=ShadowOffset, DefaultShadowOpacity := DefaultTextShadowOpacity}
    TextBlock22 : text_block = text_block{DefaultTextColor := DarkCyan, DefaultShadowOffset:=ShadowOffset, DefaultShadowOpacity := DefaultTextShadowOpacity}
    TextBlock23 : text_block = text_block{DefaultTextColor := DarkGoldenrod, DefaultShadowOffset:=ShadowOffset, DefaultShadowOpacity := DefaultTextShadowOpacity}
    TextBlock24 : text_block = text_block{DefaultTextColor := DarkGray, DefaultShadowOffset:=ShadowOffset, DefaultShadowOpacity := DefaultTextShadowOpacity}
    TextBlock25 : text_block = text_block{DefaultTextColor := DarkGreen, DefaultShadowOffset:=ShadowOffset, DefaultShadowOpacity := DefaultTextShadowOpacity}
    TextBlock26 : text_block = text_block{DefaultTextColor := DarkGrey, DefaultShadowOffset:=ShadowOffset, DefaultShadowOpacity := DefaultTextShadowOpacity}
    TextBlock27 : text_block = text_block{DefaultTextColor := DarkKhaki, DefaultShadowOffset:=ShadowOffset, DefaultShadowOpacity := DefaultTextShadowOpacity}
    TextBlock28 : text_block = text_block{DefaultTextColor := DarkMagenta, DefaultShadowOffset:=ShadowOffset, DefaultShadowOpacity := DefaultTextShadowOpacity}
    TextBlock29 : text_block = text_block{DefaultTextColor := DarkOliveGreen, DefaultShadowOffset:=ShadowOffset, DefaultShadowOpacity := DefaultTextShadowOpacity}
    TextBlock30 : text_block = text_block{DefaultTextColor := DarkOrange, DefaultShadowOffset:=ShadowOffset, DefaultShadowOpacity := DefaultTextShadowOpacity}
    TextBlock31 : text_block = text_block{DefaultTextColor := DarkOrchid, DefaultShadowOffset:=ShadowOffset, DefaultShadowOpacity := DefaultTextShadowOpacity}
    TextBlock32 : text_block = text_block{DefaultTextColor := DarkRed, DefaultShadowOffset:=ShadowOffset, DefaultShadowOpacity := DefaultTextShadowOpacity}
    TextBlock33 : text_block = text_block{DefaultTextColor := DarkSalmon, DefaultShadowOffset:=ShadowOffset, DefaultShadowOpacity := DefaultTextShadowOpacity}
    TextBlock34 : text_block = text_block{DefaultTextColor := DarkSeaGreen, DefaultShadowOffset:=ShadowOffset, DefaultShadowOpacity := DefaultTextShadowOpacity}
    TextBlock35 : text_block = text_block{DefaultTextColor := DarkSlateBlue, DefaultShadowOffset:=ShadowOffset, DefaultShadowOpacity := DefaultTextShadowOpacity}
    TextBlock36 : text_block = text_block{DefaultTextColor := DarkSlateGray, DefaultShadowOffset:=ShadowOffset, DefaultShadowOpacity := DefaultTextShadowOpacity}
    TextBlock37 : text_block = text_block{DefaultTextColor := DarkSlateGrey, DefaultShadowOffset:=ShadowOffset, DefaultShadowOpacity := DefaultTextShadowOpacity}
    TextBlock38 : text_block = text_block{DefaultTextColor := DarkTurquoise, DefaultShadowOffset:=ShadowOffset, DefaultShadowOpacity := DefaultTextShadowOpacity}
    TextBlock39 : text_block = text_block{DefaultTextColor := DarkViolet, DefaultShadowOffset:=ShadowOffset, DefaultShadowOpacity := DefaultTextShadowOpacity}
    TextBlock40 : text_block = text_block{DefaultTextColor := DeepPink, DefaultShadowOffset:=ShadowOffset, DefaultShadowOpacity := DefaultTextShadowOpacity}
    TextBlock41 : text_block = text_block{DefaultTextColor := DeepSkyBlue, DefaultShadowOffset:=ShadowOffset, DefaultShadowOpacity := DefaultTextShadowOpacity}
    TextBlock42 : text_block = text_block{DefaultTextColor := DimGray, DefaultShadowOffset:=ShadowOffset, DefaultShadowOpacity := DefaultTextShadowOpacity}
    TextBlock43 : text_block = text_block{DefaultTextColor := DimGrey, DefaultShadowOffset:=ShadowOffset, DefaultShadowOpacity := DefaultTextShadowOpacity}
    TextBlock44 : text_block = text_block{DefaultTextColor := DodgerBlue, DefaultShadowOffset:=ShadowOffset, DefaultShadowOpacity := DefaultTextShadowOpacity}
    TextBlock45 : text_block = text_block{DefaultTextColor := Firebrick, DefaultShadowOffset:=ShadowOffset, DefaultShadowOpacity := DefaultTextShadowOpacity}
    TextBlock46 : text_block = text_block{DefaultTextColor := FloralWhite, DefaultShadowOffset:=ShadowOffset, DefaultShadowOpacity := DefaultTextShadowOpacity}
    TextBlock47 : text_block = text_block{DefaultTextColor := ForestGreen, DefaultShadowOffset:=ShadowOffset, DefaultShadowOpacity := DefaultTextShadowOpacity}
    TextBlock48 : text_block = text_block{DefaultTextColor := Fuchsia, DefaultShadowOffset:=ShadowOffset, DefaultShadowOpacity := DefaultTextShadowOpacity}
    TextBlock49 : text_block = text_block{DefaultTextColor := Gainsboro, DefaultShadowOffset:=ShadowOffset, DefaultShadowOpacity := DefaultTextShadowOpacity}

    var ArrayAllTextColors : []text_block = array{}

    HUDMessageAskForHelp(Agent : agent) : void =
        UITest.SetText(GetAgentName(Agent))
        UITest.Show()

    #THIS MUST BE GO TO VFX MANAGER
    RunPetCreation(Agent : agent, PetPropAsset : creative_prop_asset, Tier : int)<suspends> : void =
        PetBlenderPropPosition:=PetBlenderProp.GetTransform().Translation
        SpawnPosition:=PetBlenderPropPosition+vector3{X:=-60.0,Y:=20.0,Z:=176.0}
        SpawnRotation:=MakeRotationFromYawPitchRollDegrees(180.0,0.0,0.0)
        #SpawnRotation:=PetBlenderProp.GetTransform().Rotation.ApplyYaw(180.0) 
        var NormalScale : vector3 = vector3{}

        if(Tier=1):
            set NormalScale = PetScale_Tier1
            #set PivotVariation = PetPivotVariation_Tier1
        if(Tier=2):
            set NormalScale = PetScale_Tier2
            #set PivotVariation = PetPivotVariation_Tier2
        if(Tier=3):
            set NormalScale = PetScale_Tier3
            #set PivotVariation = PetPivotVariation_Tier3

        PetBlenderPropTransform:=transform{Scale:=NormalScale,Rotation:=SpawnRotation,Translation:=SpawnPosition}
        ResultTuple:=SpawnProp(PetPropAsset,PetBlenderPropTransform)
        #Print("Spawning on {SpawnPosition.X} {SpawnPosition.Y} {SpawnPosition.Z}")
        if(PetProp := ResultTuple(0)?):
            SetStar(PetProp,Tier)
            RemoveHealthUI(PetProp)
            #Run pet creation
            Camera_PetCreation.AddTo(Agent)

            Frame0_VectorApplied:=vector3{X:=-16.0,Y:=0.0,Z:=-48.0}
            Frame0_RotationApplied:=MakeRotationFromYawPitchRollDegrees(0.0, -90.0, 0.0)
            Frame1_VectorApplied:=vector3{X:=-80.0,Y:=0.0,Z:=-64.0}
            Frame1_RotationApplied:=MakeRotationFromYawPitchRollDegrees(0.0, -270.0, 0.0)

            #if(PetProp.TeleportTo[PetBlenderPropPosition,SpawnRotation]):

            MovementKeyFrame0 : keyframe_delta = keyframe_delta #Key Frame 0
            {
                DeltaLocation := Frame0_VectorApplied
                #DeltaRotation := rotation{}
                DeltaRotation := Frame0_RotationApplied
                DeltaScale := NeutralScale
                Time := 0.5
                Interpolation := EaseIn
            }

            MovementKeyFrame1 : keyframe_delta = keyframe_delta #Key Frame 1
            {
                DeltaLocation := Frame1_VectorApplied
                #DeltaRotation := rotation{}
                DeltaRotation := Frame1_RotationApplied
                DeltaScale := NeutralScale
                Time := 1.0
                Interpolation := EaseIn
            }

            MovementKeyFrame2 : keyframe_delta = keyframe_delta #Key Frame 1
            {
                DeltaLocation := vector3{}
                DeltaRotation := rotation{}
                DeltaScale := NeutralScale
                Time := 1.0
                Interpolation := EaseIn
            }

            if(PropController := PetProp.GetAnimationController[]):
                # Make an array of key frames
                FortVaderKeyFrames :  []keyframe_delta = array { MovementKeyFrame0 , MovementKeyFrame1, MovementKeyFrame2}
                # Set the animation to play
                PropController.SetAnimation(FortVaderKeyFrames, ?Mode:=animation_mode.OneShot)
                PropController.Play()
                loop:
                    State := PropController.GetState()
                    if(State = animation_controller_state.Stopped):
                        break
                    Sleep(0.04)
            PetProp.Dispose()
            Camera_PetCreation.RemoveFrom(Agent)
        else:
            PrintFatalError("On RunPetCreation. Spawn result not valid")

    # Runs when the device is started in a running game

    SpawnRandomEncounter(Agent : agent, MobLB : int, Transform : transform): void =
        GeneralMobsManager.RollRandomPlantEncounter(Agent,Transform,false,MobLB,20,1)

            <#
            if(QuestInfoResult(0)?):
                #Print("QuestInfoResult(2) is {QuestInfoResult(2)}, QuestInfoResult(3) is {QuestInfoResult(3)}")
                spawn:
                    ActionUI.DisplayGather(array{QuestInfoResult(2)}, array{QuestInfoResult(3)}, array{false},false)
                BigObjetive_AudioPlayer.Play(Agent)
                spawn:
                    Delay_NewQuestDisplay(Player)
            else if(QuestInfoResult(1)?):#>
    


    #CHECK THIS TO COPY FOR THA FINAL VERSION!
               

        

            
            
    #This need to be called on the game loop
    


    <#Display_MainInventoryUI(Agent : agent) : void =
        UI:=MakePCSBInfoUI(Agent)
            UI.InitUI()
            set MaybeBakerInfoUI = option{UI}
        #Print("DISPLAYING CANVAS!!!!!!!!!!!")
        UI.Display_Canvas(5.0)#>

    <#Display_ComputerPCSBInfo(Agent : agent) : void =
        #Print("Trying to display PCSBInfo")
        #PCSBakersUpgradesUI:=MakePCSBakerUpgradeUI()
        PCSBakersUpgradesUI.SetAgent(Agent)
        PCSBakersUpgradesUI.InitUI()
        #set MaybePCSBakersUpgradesUI = option{PCSBakersUpgradesUI}
        PCSBakersUpgradesUI.SetAgent(Agent)
        PCSBakersUpgradesUI.Display_Canvas()
        PCSBakersUpgradesUI.UpdateUI()#>

    #This must be in the bakery, every bakery must have one map indicator device



    MakeColorCanvas(): canvas=
        Canvas := canvas:
            Slots:=array:
                #Main overlay
                canvas_slot:
                    Anchors:=anchors{Minimum:=vector2{X:=0.5,Y:=0.5},Maximum:=vector2{X:=0.5,Y:=0.5}}
                    ZOrder:=0
                    SizeToContent:=true
                    Offsets:=margin{Left:=0.0,Top:=0.0,Right:=0.0,Bottom:=0.0} #PositionX, PositionY, SizeX, SizeY
                    Alignment:=vector2{X:=0.5,Y:=0.5}
                    Widget:=texture_block{DefaultDesiredSize:=vector2{X:=2100.0,Y:=1250.0},DefaultImage:=Images.Slider_Difference}
                canvas_slot:
                    Anchors:=anchors{Minimum:=vector2{X:=0.5,Y:=0.5},Maximum:=vector2{X:=0.5,Y:=0.5}}
                    ZOrder:=0
                    SizeToContent:=true
                    Offsets:=margin{Left:=0.0,Top:=0.0,Right:=0.0,Bottom:=0.0} #PositionX, PositionY, SizeX, SizeY
                    Alignment:=vector2{X:=0.5,Y:=0.5}
                    Widget:=stack_box:  
                        Orientation:=orientation.Vertical
                            Slots:=array:
                                #Row Buttons
                                stack_box_slot:
                                    Widget:=stack_box:  
                                        Orientation:=orientation.Horizontal
                                        Slots:=array:
                                            stack_box_slot:
                                                Padding:=margin{Top:=2.5,Bottom:=2.5}
                                                Widget:=ChangeColorButton
                                            stack_box_slot:
                                                Padding:=margin{Top:=2.5,Bottom:=2.5}
                                                Widget:=CloseColorButton
                                #Row 0
                                stack_box_slot:
                                    Widget:=stack_box:  
                                        Orientation:=orientation.Horizontal
                                        Slots:=array:
                                            stack_box_slot:
                                                Padding:=margin{Top:=2.5,Bottom:=2.5}
                                                Widget:=TextBlock0
                                            stack_box_slot:
                                                Padding:=margin{Top:=2.5,Bottom:=2.5}
                                                Widget:=TextBlock1
                                            stack_box_slot:
                                                Padding:=margin{Top:=2.5,Bottom:=2.5}
                                                Widget:=TextBlock2
                                            stack_box_slot:
                                                Padding:=margin{Top:=2.5,Bottom:=2.5}
                                                Widget:=TextBlock3
                                            stack_box_slot:
                                                Padding:=margin{Top:=2.5,Bottom:=2.5}
                                                Widget:=TextBlock4
                                            stack_box_slot:
                                                Padding:=margin{Top:=2.5,Bottom:=2.5}
                                                Widget:=TextBlock5
                                            stack_box_slot:
                                                Padding:=margin{Top:=2.5,Bottom:=2.5}
                                                Widget:=TextBlock6
                                
                                #Row 1
                                stack_box_slot:
                                    Widget:=stack_box:  
                                        Orientation:=orientation.Horizontal
                                        Slots:=array:
                                            stack_box_slot:
                                                Padding:=margin{Top:=2.5,Bottom:=2.5}
                                                Widget:=TextBlock7
                                            stack_box_slot:
                                                Padding:=margin{Top:=2.5,Bottom:=2.5}
                                                Widget:=TextBlock8
                                            stack_box_slot:
                                                Padding:=margin{Top:=2.5,Bottom:=2.5}
                                                Widget:=TextBlock9
                                            stack_box_slot:
                                                Padding:=margin{Top:=2.5,Bottom:=2.5}
                                                Widget:=TextBlock10
                                            stack_box_slot:
                                                Padding:=margin{Top:=2.5,Bottom:=2.5}
                                                Widget:=TextBlock11
                                            stack_box_slot:
                                                Padding:=margin{Top:=2.5,Bottom:=2.5}
                                                Widget:=TextBlock12
                                            stack_box_slot:
                                                Padding:=margin{Top:=2.5,Bottom:=2.5}
                                                Widget:=TextBlock13
                                
                                #Row 2
                                stack_box_slot:
                                    Widget:=stack_box:  
                                        Orientation:=orientation.Horizontal
                                        Slots:=array:
                                            stack_box_slot:
                                                Padding:=margin{Top:=2.5,Bottom:=2.5}
                                                Widget:=TextBlock14
                                            stack_box_slot:
                                                Padding:=margin{Top:=2.5,Bottom:=2.5}
                                                Widget:=TextBlock15
                                            stack_box_slot:
                                                Padding:=margin{Top:=2.5,Bottom:=2.5}
                                                Widget:=TextBlock16
                                            stack_box_slot:
                                                Padding:=margin{Top:=2.5,Bottom:=2.5}
                                                Widget:=TextBlock17
                                            stack_box_slot:
                                                Padding:=margin{Top:=2.5,Bottom:=2.5}
                                                Widget:=TextBlock18
                                            stack_box_slot:
                                                Padding:=margin{Top:=2.5,Bottom:=2.5}
                                                Widget:=TextBlock19
                                            stack_box_slot:
                                                Padding:=margin{Top:=2.5,Bottom:=2.5}
                                                Widget:=TextBlock20
                                
                                #Row 3
                                stack_box_slot:
                                    Widget:=stack_box:  
                                        Orientation:=orientation.Horizontal
                                        Slots:=array:
                                            stack_box_slot:
                                                Padding:=margin{Top:=2.5,Bottom:=2.5}
                                                Widget:=TextBlock21
                                            stack_box_slot:
                                                Padding:=margin{Top:=2.5,Bottom:=2.5}
                                                Widget:=TextBlock22
                                            stack_box_slot:
                                                Padding:=margin{Top:=2.5,Bottom:=2.5}
                                                Widget:=TextBlock23
                                            stack_box_slot:
                                                Padding:=margin{Top:=2.5,Bottom:=2.5}
                                                Widget:=TextBlock24
                                            stack_box_slot:
                                                Padding:=margin{Top:=2.5,Bottom:=2.5}
                                                Widget:=TextBlock25
                                            stack_box_slot:
                                                Padding:=margin{Top:=2.5,Bottom:=2.5}
                                                Widget:=TextBlock26
                                            stack_box_slot:
                                                Padding:=margin{Top:=2.5,Bottom:=2.5}
                                                Widget:=TextBlock27
                                
                                #Row 4
                                stack_box_slot:
                                    Widget:=stack_box:  
                                        Orientation:=orientation.Horizontal
                                        Slots:=array:
                                            stack_box_slot:
                                                Padding:=margin{Top:=2.5,Bottom:=2.5}
                                                Widget:=TextBlock28
                                            stack_box_slot:
                                                Padding:=margin{Top:=2.5,Bottom:=2.5}
                                                Widget:=TextBlock29
                                            stack_box_slot:
                                                Padding:=margin{Top:=2.5,Bottom:=2.5}
                                                Widget:=TextBlock30
                                            stack_box_slot:
                                                Padding:=margin{Top:=2.5,Bottom:=2.5}
                                                Widget:=TextBlock31
                                            stack_box_slot:
                                                Padding:=margin{Top:=2.5,Bottom:=2.5}
                                                Widget:=TextBlock32
                                            stack_box_slot:
                                                Padding:=margin{Top:=2.5,Bottom:=2.5}
                                                Widget:=TextBlock33
                                            stack_box_slot:
                                                Padding:=margin{Top:=2.5,Bottom:=2.5}
                                                Widget:=TextBlock34
                                
                                #Row 5
                                stack_box_slot:
                                    Widget:=stack_box:  
                                        Orientation:=orientation.Horizontal
                                        Slots:=array:
                                            stack_box_slot:
                                                Padding:=margin{Top:=2.5,Bottom:=2.5}
                                                Widget:=TextBlock35
                                            stack_box_slot:
                                                Padding:=margin{Top:=2.5,Bottom:=2.5}
                                                Widget:=TextBlock36
                                            stack_box_slot:
                                                Padding:=margin{Top:=2.5,Bottom:=2.5}
                                                Widget:=TextBlock37
                                            stack_box_slot:
                                                Padding:=margin{Top:=2.5,Bottom:=2.5}
                                                Widget:=TextBlock38
                                            stack_box_slot:
                                                Padding:=margin{Top:=2.5,Bottom:=2.5}
                                                Widget:=TextBlock39
                                            stack_box_slot:
                                                Padding:=margin{Top:=2.5,Bottom:=2.5}
                                                Widget:=TextBlock40
                                            stack_box_slot:
                                                Padding:=margin{Top:=2.5,Bottom:=2.5}
                                                Widget:=TextBlock41
                                
                                #Row 6
                                stack_box_slot:
                                    Widget:=stack_box:  
                                        Orientation:=orientation.Horizontal
                                        Slots:=array:
                                            stack_box_slot:
                                                Padding:=margin{Top:=2.5,Bottom:=2.5}
                                                Widget:=TextBlock42
                                            stack_box_slot:
                                                Padding:=margin{Top:=2.5,Bottom:=2.5}
                                                Widget:=TextBlock43
                                            stack_box_slot:
                                                Padding:=margin{Top:=2.5,Bottom:=2.5}
                                                Widget:=TextBlock44
                                            stack_box_slot:
                                                Padding:=margin{Top:=2.5,Bottom:=2.5}
                                                Widget:=TextBlock45
                                            stack_box_slot:
                                                Padding:=margin{Top:=2.5,Bottom:=2.5}
                                                Widget:=TextBlock46
                                            stack_box_slot:
                                                Padding:=margin{Top:=2.5,Bottom:=2.5}
                                                Widget:=TextBlock47
                                            stack_box_slot:
                                                Padding:=margin{Top:=2.5,Bottom:=2.5}
                                                Widget:=TextBlock48
                                            

