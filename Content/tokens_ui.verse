using { /UnrealEngine.com/Temporary/UI }
using { /UnrealEngine.com/Temporary/SpatialMath }
using { /Verse.org/Simulation }
using { /Fortnite.com/Devices }
using { /Fortnite.com/UI }
using { /Verse.org/Colors/NamedColors }
using { /Verse.org/Colors }

# See https://dev.epicgames.com/documentation/en-us/uefn/create-your-own-device-in-verse for how to create a verse device.

# A Verse-authored creative device that can be placed in a level
tokens_ui := class():

    #Class settings
    var MaybeAgent : ?agent = false

    BakeryMachinesManager : bakery_machines_manager
    
    BackgroundImageSizeX : float = 1375.0
    BackgroundImageSizeY : float = 1100.0

    PTopColumn1SizeX : float = 350.0
    PTopColumn1SizeY : float = 80.0
    PTopColumn2SizeX : float = 135.0
    PTopColumn2SizeY : float = 80.0

    BUColumn1SizeX : float = 80.0
    BUColumn1SizeY : float = 80.0
    BUColumn2SizeX : float = 425.0
    BUColumn2SizeY : float = 80.0
    BUColumn3SizeX : float = 150.0
    BUColumn3SizeY : float = 80.0
    BUColumn4SizeX : float = 200.0
    BUColumn4SizeY : float = 80.0
    BUColumn5SizeX : float = 200.0
    BUColumn5SizeY : float = 80.0
    BUColumn6SizeX : float = 225.0
    BUColumn6SizeY : float = 80.0


    DT_Sapphires_Texture : texture_block = texture_block{DefaultDesiredSize:=vector2{X:=IconSmallSizeX,Y:=IconSmallSizeY}, DefaultImage:=Images.Sapphire_Icon}
    DT_TotalSapphires_BackgroundTexture : texture_block = texture_block{DefaultDesiredSize:=vector2{X:=SizeBoxSaphiresX,Y:=SizeBoxSaphiresY}, DefaultImage:=Images.Img_LineBackground2}
    DT_TotalAgedSapphires_TextBlock : text_block = text_block{DefaultTextColor:=AgedSapphiresTextColor,DefaultShadowOffset:=ShadowOffset, DefaultShadowOpacity := DefaultTextShadowOpacity}
    DT_TotalMutantMaterial_TextBlock : text_block = text_block{DefaultTextColor:=MutantMaterialTextColor,DefaultShadowOffset:=ShadowOffset, DefaultShadowOpacity := DefaultTextShadowOpacity}

    HT_Sapphires_Texture : texture_block = texture_block{DefaultDesiredSize:=vector2{X:=IconSmallSizeX,Y:=IconSmallSizeY}, DefaultImage:=Images.Sapphire_Icon}
    HT_TotalSapphires_BackgroundTexture : texture_block = texture_block{DefaultDesiredSize:=vector2{X:=SizeBoxSaphiresX,Y:=SizeBoxSaphiresY}, DefaultImage:=Images.Img_LineBackground2}
    HT_TotalAgedSapphires_TextBlock : text_block = text_block{DefaultTextColor:=AgedSapphiresTextColor,DefaultShadowOffset:=ShadowOffset, DefaultShadowOpacity := DefaultTextShadowOpacity}
    HT_TotalMutantMaterial_TextBlock : text_block = text_block{DefaultTextColor:=MutantMaterialTextColor,DefaultShadowOffset:=ShadowOffset, DefaultShadowOpacity := DefaultTextShadowOpacity}
  
    #Dungeons tokens
    var DT_MaybeCanvas : ?canvas = false
    var DT_UIIsDisplayed : logic = false

    DT_TitleTextBlock : text_block = text_block{DefaultTextColor:=White, DefaultShadowOffset:=ShadowOffset, DefaultShadowOpacity := DefaultTextShadowOpacity}

    DT_Row1_Icon : texture_block = texture_block{DefaultDesiredSize:=vector2{X:=IconSizeX,Y:=IconSizeY},DefaultImage:=Images.Bakery_Knife_Icon}
    DT_Row2_Icon : texture_block = texture_block{DefaultDesiredSize:=vector2{X:=IconSizeX,Y:=IconSizeY},DefaultImage:=Images.Bakery_Knife_Icon}
    DT_Row3_Icon : texture_block = texture_block{DefaultDesiredSize:=vector2{X:=IconSizeX,Y:=IconSizeY},DefaultImage:=Images.Bakery_Knife_Icon}
    DT_Row4_Icon : texture_block = texture_block{DefaultDesiredSize:=vector2{X:=IconSizeX,Y:=IconSizeY},DefaultImage:=Images.Bakery_Knife_Icon}
    DT_Row5_Icon : texture_block = texture_block{DefaultDesiredSize:=vector2{X:=IconSizeX,Y:=IconSizeY},DefaultImage:=Images.Bakery_Knife_Icon}

    DT_Row1_CurrentLV_TextBlock : text_block = text_block{DefaultTextColor:=CurrentTextColor, DefaultShadowOffset:=ShadowOffset, DefaultShadowOpacity := DefaultTextShadowOpacity}
    DT_Row2_CurrentLV_TextBlock : text_block = text_block{DefaultTextColor:=CurrentTextColor, DefaultShadowOffset:=ShadowOffset, DefaultShadowOpacity := DefaultTextShadowOpacity}
    DT_Row3_CurrentLV_TextBlock : text_block = text_block{DefaultTextColor:=CurrentTextColor, DefaultShadowOffset:=ShadowOffset, DefaultShadowOpacity := DefaultTextShadowOpacity}
    DT_Row4_CurrentLV_TextBlock : text_block = text_block{DefaultTextColor:=CurrentTextColor, DefaultShadowOffset:=ShadowOffset, DefaultShadowOpacity := DefaultTextShadowOpacity}
    DT_Row5_CurrentLV_TextBlock : text_block = text_block{DefaultTextColor:=CurrentTextColor, DefaultShadowOffset:=ShadowOffset, DefaultShadowOpacity := DefaultTextShadowOpacity}

    DT_Row1_Current_TextBlock : text_block = text_block{DefaultTextColor:=CurrentTextColor, DefaultShadowOffset:=ShadowOffset, DefaultShadowOpacity := DefaultTextShadowOpacity}
    DT_Row2_Current_TextBlock : text_block = text_block{DefaultTextColor:=CurrentTextColor, DefaultShadowOffset:=ShadowOffset, DefaultShadowOpacity := DefaultTextShadowOpacity}
    DT_Row3_Current_TextBlock : text_block = text_block{DefaultTextColor:=CurrentTextColor, DefaultShadowOffset:=ShadowOffset, DefaultShadowOpacity := DefaultTextShadowOpacity}
    DT_Row4_Current_TextBlock : text_block = text_block{DefaultTextColor:=CurrentTextColor, DefaultShadowOffset:=ShadowOffset, DefaultShadowOpacity := DefaultTextShadowOpacity}
    DT_Row5_Current_TextBlock : text_block = text_block{DefaultTextColor:=CurrentTextColor, DefaultShadowOffset:=ShadowOffset, DefaultShadowOpacity := DefaultTextShadowOpacity}

    DT_Row1_Upgrade_TextBlock : text_block = text_block{DefaultTextColor:=UpgradeTextColor, DefaultShadowOffset:=ShadowOffset, DefaultShadowOpacity := DefaultTextShadowOpacity}
    DT_Row2_Upgrade_TextBlock : text_block = text_block{DefaultTextColor:=UpgradeTextColor, DefaultShadowOffset:=ShadowOffset, DefaultShadowOpacity := DefaultTextShadowOpacity}
    DT_Row3_Upgrade_TextBlock : text_block = text_block{DefaultTextColor:=UpgradeTextColor, DefaultShadowOffset:=ShadowOffset, DefaultShadowOpacity := DefaultTextShadowOpacity}
    DT_Row4_Upgrade_TextBlock : text_block = text_block{DefaultTextColor:=UpgradeTextColor, DefaultShadowOffset:=ShadowOffset, DefaultShadowOpacity := DefaultTextShadowOpacity}
    DT_Row5_Upgrade_TextBlock : text_block = text_block{DefaultTextColor:=UpgradeTextColor, DefaultShadowOffset:=ShadowOffset, DefaultShadowOpacity := DefaultTextShadowOpacity}

    DT_Row1_UpgradeCost_TextBlock : text_block = text_block{DefaultTextColor:=SapphiresTextColor, DefaultShadowOffset:=ShadowOffset, DefaultShadowOpacity := DefaultTextShadowOpacity}
    DT_Row2_UpgradeCost_TextBlock : text_block = text_block{DefaultTextColor:=SapphiresTextColor, DefaultShadowOffset:=ShadowOffset, DefaultShadowOpacity := DefaultTextShadowOpacity}
    DT_Row3_UpgradeCost_TextBlock : text_block = text_block{DefaultTextColor:=SapphiresTextColor, DefaultShadowOffset:=ShadowOffset, DefaultShadowOpacity := DefaultTextShadowOpacity}
    DT_Row4_UpgradeCost_TextBlock : text_block = text_block{DefaultTextColor:=SapphiresTextColor, DefaultShadowOffset:=ShadowOffset, DefaultShadowOpacity := DefaultTextShadowOpacity}
    DT_Row5_UpgradeCost_TextBlock : text_block = text_block{DefaultTextColor:=SapphiresTextColor, DefaultShadowOffset:=ShadowOffset, DefaultShadowOpacity := DefaultTextShadowOpacity}

    DT_Row1Button : button_quiet = button_quiet{}
    DT_Row2Button : button_quiet = button_quiet{}
    DT_Row3Button : button_quiet = button_quiet{}
    DT_Row4Button : button_quiet = button_quiet{}
    DT_Row5Button : button_quiet = button_quiet{}

    #Close
    DT_CloseButton : button_loud = button_loud{}

    var AllDT_Rows_CurrentLV_TextBlock : []text_block = array{}
    var AllDT_Rows_Current_TextBlock : []text_block = array{}
    var AllDT_Rows_Upgrade_TextBlock : []text_block = array{}
    var AllDT_Rows_UpgradeCost_TextBlock : []text_block = array{}
    var AllDT_RowButtons : []button_quiet = array{}

    #hexaGameTokens
    var HT_MaybeCanvas : ?canvas = false
    var HT_UIIsDisplayed : logic = false

    HT_TitleTextBlock : text_block = text_block{DefaultTextColor:=White, DefaultShadowOffset:=ShadowOffset, DefaultShadowOpacity := DefaultTextShadowOpacity}

    HT_Row1_Icon : texture_block = texture_block{DefaultDesiredSize:=vector2{X:=IconSizeX,Y:=IconSizeY},DefaultImage:=Images.Bakery_Knife_Icon}
    HT_Row2_Icon : texture_block = texture_block{DefaultDesiredSize:=vector2{X:=IconSizeX,Y:=IconSizeY},DefaultImage:=Images.Bakery_Knife_Icon}
    HT_Row3_Icon : texture_block = texture_block{DefaultDesiredSize:=vector2{X:=IconSizeX,Y:=IconSizeY},DefaultImage:=Images.Bakery_Knife_Icon}
    HT_Row4_Icon : texture_block = texture_block{DefaultDesiredSize:=vector2{X:=IconSizeX,Y:=IconSizeY},DefaultImage:=Images.Bakery_Knife_Icon}
    HT_Row5_Icon : texture_block = texture_block{DefaultDesiredSize:=vector2{X:=IconSizeX,Y:=IconSizeY},DefaultImage:=Images.Bakery_Knife_Icon}

    HT_Row1_CurrentLV_TextBlock : text_block = text_block{DefaultTextColor:=CurrentTextColor, DefaultShadowOffset:=ShadowOffset, DefaultShadowOpacity := DefaultTextShadowOpacity}
    HT_Row2_CurrentLV_TextBlock : text_block = text_block{DefaultTextColor:=CurrentTextColor, DefaultShadowOffset:=ShadowOffset, DefaultShadowOpacity := DefaultTextShadowOpacity}
    HT_Row3_CurrentLV_TextBlock : text_block = text_block{DefaultTextColor:=CurrentTextColor, DefaultShadowOffset:=ShadowOffset, DefaultShadowOpacity := DefaultTextShadowOpacity}
    HT_Row4_CurrentLV_TextBlock : text_block = text_block{DefaultTextColor:=CurrentTextColor, DefaultShadowOffset:=ShadowOffset, DefaultShadowOpacity := DefaultTextShadowOpacity}
    HT_Row5_CurrentLV_TextBlock : text_block = text_block{DefaultTextColor:=CurrentTextColor, DefaultShadowOffset:=ShadowOffset, DefaultShadowOpacity := DefaultTextShadowOpacity}

    HT_Row1_Current_TextBlock : text_block = text_block{DefaultTextColor:=CurrentTextColor, DefaultShadowOffset:=ShadowOffset, DefaultShadowOpacity := DefaultTextShadowOpacity}
    HT_Row2_Current_TextBlock : text_block = text_block{DefaultTextColor:=CurrentTextColor, DefaultShadowOffset:=ShadowOffset, DefaultShadowOpacity := DefaultTextShadowOpacity}
    HT_Row3_Current_TextBlock : text_block = text_block{DefaultTextColor:=CurrentTextColor, DefaultShadowOffset:=ShadowOffset, DefaultShadowOpacity := DefaultTextShadowOpacity}
    HT_Row4_Current_TextBlock : text_block = text_block{DefaultTextColor:=CurrentTextColor, DefaultShadowOffset:=ShadowOffset, DefaultShadowOpacity := DefaultTextShadowOpacity}
    HT_Row5_Current_TextBlock : text_block = text_block{DefaultTextColor:=CurrentTextColor, DefaultShadowOffset:=ShadowOffset, DefaultShadowOpacity := DefaultTextShadowOpacity}

    HT_Row1_Upgrade_TextBlock : text_block = text_block{DefaultTextColor:=UpgradeTextColor, DefaultShadowOffset:=ShadowOffset, DefaultShadowOpacity := DefaultTextShadowOpacity}
    HT_Row2_Upgrade_TextBlock : text_block = text_block{DefaultTextColor:=UpgradeTextColor, DefaultShadowOffset:=ShadowOffset, DefaultShadowOpacity := DefaultTextShadowOpacity}
    HT_Row3_Upgrade_TextBlock : text_block = text_block{DefaultTextColor:=UpgradeTextColor, DefaultShadowOffset:=ShadowOffset, DefaultShadowOpacity := DefaultTextShadowOpacity}
    HT_Row4_Upgrade_TextBlock : text_block = text_block{DefaultTextColor:=UpgradeTextColor, DefaultShadowOffset:=ShadowOffset, DefaultShadowOpacity := DefaultTextShadowOpacity}
    HT_Row5_Upgrade_TextBlock : text_block = text_block{DefaultTextColor:=UpgradeTextColor, DefaultShadowOffset:=ShadowOffset, DefaultShadowOpacity := DefaultTextShadowOpacity}

    HT_Row1_UpgradeCost_TextBlock : text_block = text_block{DefaultTextColor:=SapphiresTextColor, DefaultShadowOffset:=ShadowOffset, DefaultShadowOpacity := DefaultTextShadowOpacity}
    HT_Row2_UpgradeCost_TextBlock : text_block = text_block{DefaultTextColor:=SapphiresTextColor, DefaultShadowOffset:=ShadowOffset, DefaultShadowOpacity := DefaultTextShadowOpacity}
    HT_Row3_UpgradeCost_TextBlock : text_block = text_block{DefaultTextColor:=SapphiresTextColor, DefaultShadowOffset:=ShadowOffset, DefaultShadowOpacity := DefaultTextShadowOpacity}
    HT_Row4_UpgradeCost_TextBlock : text_block = text_block{DefaultTextColor:=SapphiresTextColor, DefaultShadowOffset:=ShadowOffset, DefaultShadowOpacity := DefaultTextShadowOpacity}
    HT_Row5_UpgradeCost_TextBlock : text_block = text_block{DefaultTextColor:=SapphiresTextColor, DefaultShadowOffset:=ShadowOffset, DefaultShadowOpacity := DefaultTextShadowOpacity}

    HT_Row1Button : button_quiet = button_quiet{}
    HT_Row2Button : button_quiet = button_quiet{}
    HT_Row3Button : button_quiet = button_quiet{}
    HT_Row4Button : button_quiet = button_quiet{}
    HT_Row5Button : button_quiet = button_quiet{}

    var AllHT_Rows_CurrentLV_TextBlock : []text_block = array{}
    var AllHT_Rows_Current_TextBlock : []text_block = array{}
    var AllHT_Rows_Upgrade_TextBlock : []text_block = array{}
    var AllHT_Rows_UpgradeCost_TextBlock : []text_block = array{}
    var AllHT_RowButtons : []button_quiet = array{}

    #Close
    HT_CloseButton : button_loud = button_loud{}

    InitUI() : void =

        DT_TitleTextBlock.SetText(BasicMessage("Dungeon perks"))
        HT_TitleTextBlock.SetText(BasicMessage("Master competition perks"))

        SetArrays()
        InitSubscribers()

        DT_Canvas := MakeDungeonPerksCanvas()
        set DT_MaybeCanvas = option{DT_Canvas}

        HT_Canvas := MakeHexaGamePerksCanvas()
        set HT_MaybeCanvas = option{HT_Canvas}

    SetArrays() : void =
        
        #Dungeon tokens
        set AllDT_Rows_CurrentLV_TextBlock = array{DT_Row1_CurrentLV_TextBlock,DT_Row2_CurrentLV_TextBlock,DT_Row3_CurrentLV_TextBlock,DT_Row4_CurrentLV_TextBlock, DT_Row5_CurrentLV_TextBlock}

        set AllDT_Rows_Current_TextBlock = array{DT_Row1_Current_TextBlock,DT_Row2_Current_TextBlock,DT_Row3_Current_TextBlock,DT_Row4_Current_TextBlock, DT_Row5_Current_TextBlock}
        set AllDT_Rows_Upgrade_TextBlock = array{DT_Row1_Upgrade_TextBlock,DT_Row2_Upgrade_TextBlock,DT_Row3_Upgrade_TextBlock,DT_Row4_Upgrade_TextBlock, DT_Row5_Upgrade_TextBlock}
        set AllDT_Rows_UpgradeCost_TextBlock = array{DT_Row1_UpgradeCost_TextBlock,DT_Row2_UpgradeCost_TextBlock,DT_Row3_UpgradeCost_TextBlock,DT_Row4_UpgradeCost_TextBlock, DT_Row5_UpgradeCost_TextBlock}

        set AllDT_RowButtons = array{DT_Row1Button,DT_Row2Button,DT_Row3Button,DT_Row4Button,DT_Row5Button} 

        #Hexagame tokens
        set AllHT_Rows_CurrentLV_TextBlock = array{HT_Row1_CurrentLV_TextBlock,HT_Row2_CurrentLV_TextBlock,HT_Row3_CurrentLV_TextBlock,HT_Row4_CurrentLV_TextBlock, HT_Row5_CurrentLV_TextBlock}

        set AllHT_Rows_Current_TextBlock = array{HT_Row1_Current_TextBlock,HT_Row2_Current_TextBlock,HT_Row3_Current_TextBlock,HT_Row4_Current_TextBlock, HT_Row5_Current_TextBlock}
        set AllHT_Rows_Upgrade_TextBlock = array{HT_Row1_Upgrade_TextBlock,HT_Row2_Upgrade_TextBlock,HT_Row3_Upgrade_TextBlock,HT_Row4_Upgrade_TextBlock, HT_Row5_Upgrade_TextBlock}
        set AllHT_Rows_UpgradeCost_TextBlock = array{HT_Row1_UpgradeCost_TextBlock,HT_Row2_UpgradeCost_TextBlock,HT_Row3_UpgradeCost_TextBlock,HT_Row4_UpgradeCost_TextBlock, HT_Row5_UpgradeCost_TextBlock}

        set AllHT_RowButtons = array{HT_Row1Button,HT_Row2Button,HT_Row3Button,HT_Row4Button,HT_Row5Button}

    InitSubscribers() : void =
        DT_CloseButton.OnClick().Subscribe(DT_Remove_Canvas) 
        HT_CloseButton.OnClick().Subscribe(HT_Remove_Canvas) 

        DT_Row1Button.OnClick().Subscribe(handler_weaponsandmobs{MaybeTokensUIManager:=option{Self}, DungeonToken_Perk:=DUNGEONPERK_DELAY}.HandlerDTTokensUI)
        DT_Row2Button.OnClick().Subscribe(handler_weaponsandmobs{MaybeTokensUIManager:=option{Self}, DungeonToken_Perk:=DUNGEONPERK_ACCUMULATED}.HandlerDTTokensUI)

        HT_Row1Button.OnClick().Subscribe(handler_weaponsandmobs{MaybeTokensUIManager:=option{Self}, DungeonToken_Perk:=HEXAGAMEPERK_DELAY}.HandlerHTTokensUI)
        HT_Row2Button.OnClick().Subscribe(handler_weaponsandmobs{MaybeTokensUIManager:=option{Self}, DungeonToken_Perk:=HEXAGAMEPERK_ACCUMULATED}.HandlerHTTokensUI)

    TryImproveDTPerk(Perk : int) : void =
        if(Agent := MaybeAgent?,Player := player[Agent], Player.IsActive[], PlayerData:=BakeryMachinesManager.PlayerDataCache):
            var MaxLevel : int = 0
            var LevelPerk : int = 0

            if(Perk=DUNGEONPERK_DELAY):
                set MaxLevel = MainFormula_MaxLVDelayDungeon()
                set LevelPerk = PlayerData.Dungeon_LVDelay
            else if(Perk=DUNGEONPERK_ACCUMULATED):
                set MaxLevel = Dungeon_MaxToken_Accumulated
                set LevelPerk = PlayerData.Dungeon_LVAccumulated

            TupleResult := CalculeDungeonTokenStat(Perk, LevelPerk)

            #Print("MaxLevel: {MaxLevel}, LVDungeonTokenPerk: {LevelPerk}")
            #Print("Cost {GetMagnitudeLetter(TupleResult(1),"","")}")

            BakeryMachinesManager.CheckQuestProgress(1.0,AGGEDSAPPHIRESUSED_TUTORIAL_TYPE)

            if(Perk = DUNGEONPERK_DELAY):
                if(TupleResult(1)<=PlayerData.Inventory_MutantMaterial, MaxLevel>LevelPerk):
                    TableWithoutMutantMaterial:=BakeryMachinesManager.GameManager.PlayersDataManager.WithdrawResource(PlayerData,TupleResult(1),MUTANTMATERIAL)
                    TableWithNewDelay:=BakeryMachinesManager.GameManager.PlayersDataManager.DecrementDungeon_Delay(TableWithoutMutantMaterial,-TupleResult(2))
                    FinalTable:=BakeryMachinesManager.GameManager.PlayersDataManager.ImproveDungeonPerk(TableWithNewDelay,Perk)
                    BakeryMachinesManager.SetNewPlayerDataCache(FinalTable)
                    #BakeryMachinesManager.GameManager.PlayersDataManager.SaveAllNewData(Player,PlayerData,FinalTable,BakeryMachinesManager)
                    BakeryMachinesManager.CheckQuestProgress(1.0,TOKENIMPROVE_TUTORIAL_TYPE)
                    DT_UpdateUI()
            else if(TupleResult(1)<=PlayerData.Inventory_AgedSapphires, MaxLevel>LevelPerk):
                TableWithoutSapphires:=BakeryMachinesManager.GameManager.PlayersDataManager.WithdrawResource(PlayerData,TupleResult(1),AGEDSAPPHIRES)
                FinalTable:=BakeryMachinesManager.GameManager.PlayersDataManager.ImproveDungeonPerk(TableWithoutSapphires,Perk)
                BakeryMachinesManager.SetNewPlayerDataCache(FinalTable)
                #BakeryMachinesManager.GameManager.PlayersDataManager.SaveAllNewData(Player,PlayerData,FinalTable,BakeryMachinesManager)
                BakeryMachinesManager.CheckQuestProgress(1.0,TOKENIMPROVE_TUTORIAL_TYPE)
                DT_UpdateUI()
            else:
                spawn:
                    BakeryMachinesManager.GameManager.VFXManager.PlaySound(ERROR_SOUND,0.0,option{Agent},false)

    TryImproveHTPerk(Perk : int) : void =
        if(Agent := MaybeAgent?,Player := player[Agent], Player.IsActive[], PlayerData:=BakeryMachinesManager.PlayerDataCache):
            var MaxLevel : int = 0
            var LevelPerk : int = 0

            if(Perk=HEXAGAMEPERK_DELAY):
                set MaxLevel = MainFormula_MaxLVDelayHexaGame()
                set LevelPerk = PlayerData.HexaGame_LVDelay
            else if(Perk=HEXAGAMEPERK_ACCUMULATED):
                set MaxLevel = HexaGame_MaxToken_Accumulated
                set LevelPerk = PlayerData.HexaGame_LVAccumulated

            TupleResult := CalculeHexaGameTokenStat(Perk, LevelPerk)

            if(Perk = HEXAGAMEPERK_DELAY):
                if(TupleResult(1)<=PlayerData.Inventory_MutantMaterial, MaxLevel>LevelPerk):
                    TableWithoutMutantMaterial:=BakeryMachinesManager.GameManager.PlayersDataManager.WithdrawResource(PlayerData,TupleResult(1),MUTANTMATERIAL)
                    TableWithNewDelay:=BakeryMachinesManager.GameManager.PlayersDataManager.DecrementHexaGame_Delay(TableWithoutMutantMaterial,-TupleResult(2))
                    FinalTable:=BakeryMachinesManager.GameManager.PlayersDataManager.ImproveHexaGamePerk(TableWithNewDelay,Perk)
                    BakeryMachinesManager.SetNewPlayerDataCache(FinalTable)
                    #BakeryMachinesManager.GameManager.PlayersDataManager.SaveAllNewData(Player,PlayerData,FinalTable,BakeryMachinesManager)
                    BakeryMachinesManager.CheckQuestProgress(1.0,TOKENIMPROVE_TUTORIAL_TYPE)
                    HT_UpdateUI()
            else if(TupleResult(1)<=PlayerData.Inventory_AgedSapphires, MaxLevel>LevelPerk):
                TableWithoutSapphires:=BakeryMachinesManager.GameManager.PlayersDataManager.WithdrawResource(PlayerData,TupleResult(1),AGEDSAPPHIRES)
                FinalTable:=BakeryMachinesManager.GameManager.PlayersDataManager.ImproveHexaGamePerk(TableWithoutSapphires,Perk)
                BakeryMachinesManager.SetNewPlayerDataCache(FinalTable)
                #BakeryMachinesManager.GameManager.PlayersDataManager.SaveAllNewData(Player,PlayerData,FinalTable,BakeryMachinesManager)
                BakeryMachinesManager.CheckQuestProgress(1.0,TOKENIMPROVE_TUTORIAL_TYPE)
                HT_UpdateUI()
            else:
                spawn:
                    BakeryMachinesManager.GameManager.VFXManager.PlaySound(ERROR_SOUND,0.0,option{Agent},false)

    SetAgent(InAgent : agent) : void =
        set DT_UIIsDisplayed = false
        set HT_UIIsDisplayed = false
        set MaybeAgent = option{InAgent}

    DT_UpdateUI() : void =

        if(Agent := MaybeAgent?,Player := player[Agent], Player.IsActive[], PlayerData:=BakeryMachinesManager.PlayerDataCache):

            DT_CloseButton.SetText(BasicMessage("CLOSE"))

            #Inventory resourcers
            DT_TotalMutantMaterial_TextBlock.SetText(BasicMessage("{GetOptimizedMagnitudeLetter(PlayerData.Inventory_MutantMaterial)}"))
            DT_TotalAgedSapphires_TextBlock.SetText(BasicMessage("{GetOptimizedMagnitudeLetter(PlayerData.Inventory_AgedSapphires)}"))

            for(Perk:=DUNGEONPERK_DELAY..DUNGEONPERK_ACCUMULATED):
                if:
                    DT_Row_Current_TextBlock := AllDT_Rows_Current_TextBlock[Perk]
                    DT_Row_Upgrade_TextBlock := AllDT_Rows_Upgrade_TextBlock[Perk]
                    DT_Row_UpgradeCost_TextBlock := AllDT_Rows_UpgradeCost_TextBlock[Perk]
                then:
                    var StatLV : int = 0

                    var TupleResult : tuple(float,float,float) = (0.0,0.0,0.0)
                    var MaxLevel : int = 0       
                        
                    if(Perk = DUNGEONPERK_DELAY):
                        set TupleResult = CalculeDungeonTokenStat(Perk, PlayerData.Dungeon_LVDelay)
                        set MaxLevel = MainFormula_MaxLVDelayDungeon()
                        set StatLV = PlayerData.Dungeon_LVDelay
                    else if(Perk = DUNGEONPERK_ACCUMULATED):
                        set TupleResult = CalculeDungeonTokenStat(Perk, PlayerData.Dungeon_LVAccumulated)
                        set MaxLevel = Dungeon_MaxToken_Accumulated
                        set StatLV = PlayerData.Dungeon_LVAccumulated
                    
                    if(StatLV<MaxLevel):
                        if(Perk = DUNGEONPERK_ACCUMULATED):
                            DT_Row_UpgradeCost_TextBlock.SetText(BasicMessage(GetMagnitudeLetter(TupleResult(1),"x","")))
                            if:
                                ToFloorUpgrade := Floor[TupleResult(2)]
                                ToFloorCurrent := Floor[TupleResult(0)]
                            then:
                                DT_Row_Upgrade_TextBlock.SetText(BasicMessage("+{ToFloorUpgrade}"))
                                DT_Row_Current_TextBlock.SetText(BasicMessage("{ToFloorCurrent}"))
                        else if(Perk = DUNGEONPERK_DELAY):
                            DT_Row_UpgradeCost_TextBlock.SetText(BasicMessage(GetMagnitudeLetter(TupleResult(1),"x","")))
                            DT_Row_Upgrade_TextBlock.SetText(BasicMessage(GetMagnitudeLetter(TupleResult(2),"","s")))
                            DT_Row_Current_TextBlock.SetText(BasicMessage(GetMagnitudeLetter(TupleResult(0),"","s")))

                        DT_Row_Current_TextBlock.SetTextColor(CurrentTextColor)
                        DT_Row_Upgrade_TextBlock.SetTextColor(UpgradeTextColor)
                    
                        if(Perk = DUNGEONPERK_ACCUMULATED):
                            if(TupleResult(1)<=PlayerData.Inventory_AgedSapphires):
                                DT_Row_UpgradeCost_TextBlock.SetTextColor(AgedSapphiresTextColor)
                            else:
                                DT_Row_UpgradeCost_TextBlock.SetTextColor(NotEnoughTextColor)
                        else if(TupleResult(1)<=PlayerData.Inventory_MutantMaterial):
                            DT_Row_UpgradeCost_TextBlock.SetTextColor(MutantMaterialTextColor)
                        else:
                            DT_Row_UpgradeCost_TextBlock.SetTextColor(NotEnoughTextColor) 
                    else:
                        if(Perk = DUNGEONPERK_DELAY):
                            DT_Row_Current_TextBlock.SetText(BasicMessage(GetMagnitudeLetter(TupleResult(0),"","s")))
                        else if(Perk = DUNGEONPERK_ACCUMULATED):
                            if:
                                ToFloorCurrent := Floor[TupleResult(0)]
                            then:
                                DT_Row_Current_TextBlock.SetText(BasicMessage("{ToFloorCurrent}"))

                        DT_Row_Upgrade_TextBlock.SetText(BasicMessage("MAXED"))
                        DT_Row_Upgrade_TextBlock.SetTextColor(MaxedColor)
                        DT_Row_UpgradeCost_TextBlock.SetText(BasicMessage("MAXED"))
                        DT_Row_UpgradeCost_TextBlock.SetTextColor(MaxedColor)

    HT_UpdateUI() : void =

        if(Agent := MaybeAgent?,Player := player[Agent], Player.IsActive[], PlayerData:=BakeryMachinesManager.PlayerDataCache):

            HT_CloseButton.SetText(BasicMessage("CLOSE"))

            #Inventory resourcers
            HT_TotalMutantMaterial_TextBlock.SetText(BasicMessage("{GetOptimizedMagnitudeLetter(PlayerData.Inventory_MutantMaterial)}"))
            HT_TotalAgedSapphires_TextBlock.SetText(BasicMessage("{GetOptimizedMagnitudeLetter(PlayerData.Inventory_AgedSapphires)}"))

            for(Perk:=HEXAGAMEPERK_DELAY..HEXAGAMEPERK_ACCUMULATED):
                if:
                    HT_Row_Current_TextBlock := AllHT_Rows_Current_TextBlock[Perk]
                    HT_Row_Upgrade_TextBlock := AllHT_Rows_Upgrade_TextBlock[Perk]
                    HT_Row_UpgradeCost_TextBlock := AllHT_Rows_UpgradeCost_TextBlock[Perk]
                then:
                    var StatLV : int = 0

                    var TupleResult : tuple(float,float,float) = (0.0,0.0,0.0)
                    var MaxLevel : int = 0       
                        
                    if(Perk = HEXAGAMEPERK_DELAY):
                        set TupleResult = CalculeHexaGameTokenStat(Perk, PlayerData.HexaGame_LVDelay)
                        set MaxLevel = MainFormula_MaxLVDelayHexaGame()
                        set StatLV = PlayerData.HexaGame_LVDelay
                    else if(Perk = HEXAGAMEPERK_ACCUMULATED):
                        set TupleResult = CalculeHexaGameTokenStat(Perk, PlayerData.HexaGame_LVAccumulated)
                        set MaxLevel = HexaGame_MaxToken_Accumulated
                        set StatLV = PlayerData.HexaGame_LVAccumulated
                    
                    if(StatLV<MaxLevel):
                        if(Perk = HEXAGAMEPERK_ACCUMULATED):
                            HT_Row_UpgradeCost_TextBlock.SetText(BasicMessage(GetMagnitudeLetter(TupleResult(1),"x","")))
                            if:
                                ToFloorUpgrade := Floor[TupleResult(2)]
                                ToFloorCurrent := Floor[TupleResult(0)]
                            then:
                                HT_Row_Upgrade_TextBlock.SetText(BasicMessage("+{ToFloorUpgrade}"))
                                HT_Row_Current_TextBlock.SetText(BasicMessage("{ToFloorCurrent}"))
                        else if(Perk = HEXAGAMEPERK_DELAY):
                            HT_Row_UpgradeCost_TextBlock.SetText(BasicMessage(GetMagnitudeLetter(TupleResult(1),"x","")))
                            HT_Row_Upgrade_TextBlock.SetText(BasicMessage(GetMagnitudeLetter(TupleResult(2),"","s")))
                            HT_Row_Current_TextBlock.SetText(BasicMessage(GetMagnitudeLetter(TupleResult(0),"","s")))

                        HT_Row_Current_TextBlock.SetTextColor(CurrentTextColor)
                        HT_Row_Upgrade_TextBlock.SetTextColor(UpgradeTextColor)
                    
                        if(Perk = HEXAGAMEPERK_ACCUMULATED):
                            if(TupleResult(1)<=PlayerData.Inventory_AgedSapphires):
                                HT_Row_UpgradeCost_TextBlock.SetTextColor(AgedSapphiresTextColor)
                            else:
                                HT_Row_UpgradeCost_TextBlock.SetTextColor(NotEnoughTextColor)
                        else if(TupleResult(1)<=PlayerData.Inventory_MutantMaterial):
                            HT_Row_UpgradeCost_TextBlock.SetTextColor(MutantMaterialTextColor)
                        else:
                            HT_Row_UpgradeCost_TextBlock.SetTextColor(NotEnoughTextColor) 
                    else:
                        if(Perk = HEXAGAMEPERK_DELAY):
                            HT_Row_Current_TextBlock.SetText(BasicMessage(GetMagnitudeLetter(TupleResult(0),"","s")))
                        else if(Perk = HEXAGAMEPERK_ACCUMULATED):
                            if:
                                ToFloorCurrent := Floor[TupleResult(0)]
                            then:
                                HT_Row_Current_TextBlock.SetText(BasicMessage("{ToFloorCurrent}"))

                        HT_Row_Upgrade_TextBlock.SetText(BasicMessage("MAXED"))
                        HT_Row_Upgrade_TextBlock.SetTextColor(MaxedColor)
                        HT_Row_UpgradeCost_TextBlock.SetText(BasicMessage("MAXED"))
                        HT_Row_UpgradeCost_TextBlock.SetTextColor(MaxedColor)

    DT_Remove_Canvas(Message : widget_message) : void =
        set DT_UIIsDisplayed = false
        if:
            Agent:=MaybeAgent?
            Player:=player[Agent]
            PlayerUI:=GetPlayerUI[Player]
            Canvas := DT_MaybeCanvas?
        then:
            PlayerUI.RemoveWidget(Canvas)
            DT_RemoveDelay_OnRemove()

    HT_Remove_Canvas(Message : widget_message) : void =
        set HT_UIIsDisplayed = false
        if:
            Agent:=MaybeAgent?
            Player:=player[Agent]
            PlayerUI:=GetPlayerUI[Player]
            Canvas := HT_MaybeCanvas?
        then:
            PlayerUI.RemoveWidget(Canvas)
            HT_RemoveDelay_OnRemove()

    DT_Remove_CanvasWithoutMessage() : void =
        set HT_UIIsDisplayed = false
        if:
            Agent:=MaybeAgent?
            Player:=player[Agent]
            PlayerUI:=GetPlayerUI[Player]
            Canvas := DT_MaybeCanvas?
        then:
            PlayerUI.RemoveWidget(Canvas)
            DT_RemoveDelay_OnRemove()

    HT_Remove_CanvasWithoutMessage() : void =
        set HT_UIIsDisplayed = false
        if:
            Agent:=MaybeAgent?
            Player:=player[Agent]
            PlayerUI:=GetPlayerUI[Player]
            Canvas := HT_MaybeCanvas?
        then:
            PlayerUI.RemoveWidget(Canvas)
            HT_RemoveDelay_OnRemove()

    HT_RemoveDelay_OnAdd() : void =
        HT_CloseButton.SetText(BasicMessage("Close"))
        #CurrentBuyOption_textblock.SetVisibility(widget_visibility.Hidden)

    HT_RemoveDelay_OnRemove() : void =
        HT_CloseButton.SetText(BasicMessage("Close "))
        #CurrentBuyOption_textblock.SetVisibility(widget_visibility.Hidden)

    DT_RemoveDelay_OnAdd() : void =
        DT_CloseButton.SetText(BasicMessage("Close"))
        #CurrentBuyOption_textblock.SetVisibility(widget_visibility.Hidden)

    DT_RemoveDelay_OnRemove() : void =
        DT_CloseButton.SetText(BasicMessage("Close "))
        #CurrentBuyOption_textblock.SetVisibility(widget_visibility.Hidden)

            
    DT_Display_Canvas() : void =
        if(not DT_UIIsDisplayed?):
            set DT_UIIsDisplayed = true
            if:
                Agent:=MaybeAgent?
                PlayerUI:=GetPlayerUI[player[Agent]]
                Canvas := DT_MaybeCanvas?
            then:
                PlayerUI.AddWidget(Canvas,player_ui_slot{InputMode := ui_input_mode.All, ZOrder:=1}) 
                DT_RemoveDelay_OnAdd() 
                DT_UpdateUI()

    HT_Display_Canvas() : void =
        if(not HT_UIIsDisplayed?):
            set HT_UIIsDisplayed = true
            if:
                Agent:=MaybeAgent?
                PlayerUI:=GetPlayerUI[player[Agent]]
                Canvas := HT_MaybeCanvas?
            then:
                PlayerUI.AddWidget(Canvas,player_ui_slot{InputMode := ui_input_mode.All, ZOrder:=1}) 
                HT_RemoveDelay_OnAdd() 
                HT_UpdateUI()

    MakeHexaGamePerksCanvas(): canvas=
        Canvas := canvas:
            Slots:=array:
                #Background black image
                canvas_slot:
                    Anchors:=anchors{Minimum:=vector2{X:=0.5,Y:=0.5},Maximum:=vector2{X:=0.5,Y:=0.5}}
                    ZOrder:=0
                    Alignment:=vector2{X:=0.5,Y:=0.5}
                    Widget:=texture_block:
                        DefaultImage:=Images.BlackBackground
                        DefaultDesiredSize:=vector2{X:=BackgroundBlackUISizeX,Y:=BackgroundBlackUISizeY}
                #Background image
                canvas_slot:
                    Anchors:=anchors{Minimum:=vector2{X:=0.5,Y:=0.5},Maximum:=vector2{X:=0.5,Y:=0.5}}
                    ZOrder:=0
                    Offsets:=margin{Left:=0.0,Top:=0.0,Right:=BackgroundImageSizeX,Bottom:=BackgroundImageSizeY} #PositionX, PositionY, SizeX, SizeY
                    Alignment:=vector2{X:=0.5,Y:=0.5}
                    Widget:=texture_block:
                        DefaultImage:=Images.Img_MainBackground
                        DefaultDesiredSize:=vector2{X:=BackgroundImageSizeX,Y:=BackgroundImageSizeY}
                #Main overlay
                canvas_slot:
                    Anchors:=anchors{Minimum:=vector2{X:=0.5,Y:=0.5},Maximum:=vector2{X:=0.5,Y:=0.5}}
                    ZOrder:=0
                    Alignment:=vector2{X:=0.5,Y:=0.5}
                    Widget:=overlay:
                        Slots:=array:
                            overlay_slot:
                                HorizontalAlignment:=horizontal_alignment.Center
                                VerticalAlignment:=vertical_alignment.Center
                                Padding:=margin{Left:=0.0,Top:=10.0,Right:=0.0,Bottom:=0.0}
                                Widget:=stack_box:
                                    Orientation:=orientation.Vertical
                                        Slots:=array:
                                            stack_box_slot: #Sapphires per produc
                                                HorizontalAlignment:=horizontal_alignment.Right
                                                Widget:=stack_box:
                                                    Orientation:=orientation.Horizontal
                                                        Slots:=array:
                                                            stack_box_slot: #Current MUTANTMATERIAL
                                                                Padding:=margin{Left:=InterColumnLeftPadding,Top:=InterColumnTopPadding,Right:=InterColumnRightPadding,Bottom:=InterColumnBottomPadding}
                                                                Widget:=overlay:
                                                                    Slots:=array:
                                                                        overlay_slot:
                                                                            Widget:= texture_block{DefaultDesiredSize:=vector2{X:=SizeBoxSaphiresX,Y:=SizeBoxSaphiresY}, DefaultImage:=Images.Img_LineBackground2}
                                                                        overlay_slot:
                                                                            HorizontalAlignment:=horizontal_alignment.Left
                                                                            VerticalAlignment:=vertical_alignment.Center 
                                                                            Padding:=margin{Left:=10.0,Top:=0.0,Right:=0.0,Bottom:=0.0}
                                                                            Widget:=texture_block{DefaultDesiredSize:=vector2{X:=IconSmallSizeX,Y:=IconSmallSizeY}, DefaultImage:=Images.Mutant_Material_Icon}
                                                                        overlay_slot: 
                                                                            HorizontalAlignment:=horizontal_alignment.Left
                                                                            VerticalAlignment:=vertical_alignment.Center 
                                                                            Padding:=margin{Left:=65.0,Top:=0.0,Right:=0.0,Bottom:=0.0}
                                                                            Widget:=HT_TotalMutantMaterial_TextBlock
                                                            stack_box_slot: #Current AGEDSAPPHIRES
                                                                Padding:=margin{Left:=InterColumnLeftPadding,Top:=InterColumnTopPadding,Right:=InterColumnRightPadding,Bottom:=InterColumnBottomPadding}
                                                                Widget:=overlay:
                                                                    Slots:=array:
                                                                        overlay_slot:
                                                                            Widget:= texture_block{DefaultDesiredSize:=vector2{X:=SizeBoxSaphiresX,Y:=SizeBoxSaphiresY}, DefaultImage:=Images.Img_LineBackground2}
                                                                        overlay_slot:
                                                                            HorizontalAlignment:=horizontal_alignment.Left
                                                                            VerticalAlignment:=vertical_alignment.Center 
                                                                            Padding:=margin{Left:=10.0,Top:=0.0,Right:=0.0,Bottom:=0.0}
                                                                            Widget:=texture_block{DefaultDesiredSize:=vector2{X:=IconSmallSizeX,Y:=IconSmallSizeY}, DefaultImage:=Images.AgedSaphire_Icon}
                                                                        overlay_slot: 
                                                                            HorizontalAlignment:=horizontal_alignment.Left
                                                                            VerticalAlignment:=vertical_alignment.Center 
                                                                            Padding:=margin{Left:=65.0,Top:=0.0,Right:=0.0,Bottom:=0.0}
                                                                            Widget:=HT_TotalAgedSapphires_TextBlock       
                                            stack_box_slot: #Perk upgrades
                                                Widget:=stack_box:
                                                    Orientation:=orientation.Vertical
                                                        Slots:=array:
                                                            stack_box_slot: #Line 0: All titles
                                                                HorizontalAlignment:=horizontal_alignment.Right
                                                                Padding:=margin{Left:=RowLeftPadding,Top:=25.0,Right:=RowRightPadding,Bottom:=RowBottomPadding}
                                                                    Widget:=overlay:
                                                                        Slots:=array:
                                                                            overlay_slot: 
                                                                                HorizontalAlignment:=horizontal_alignment.Fill
                                                                                VerticalAlignment:=vertical_alignment.Center
                                                                                Widget:=texture_block{DefaultDesiredSize:=vector2{X:=0.0,Y:=LineSizeY},DefaultImage:=Images.Img_LineBackground}
                                                            stack_box_slot: #Line 1: Title
                                                                Padding:=margin{Left:=RowLeftPadding,Top:=RowTopPadding,Right:=RowRightPadding,Bottom:=RowBottomPadding}
                                                                HorizontalAlignment:=horizontal_alignment.Fill
                                                                VerticalAlignment:=vertical_alignment.Bottom
                                                                Widget:=overlay:
                                                                    Slots:=array:
                                                                        overlay_slot: 
                                                                            HorizontalAlignment:=horizontal_alignment.Fill
                                                                            VerticalAlignment:=vertical_alignment.Bottom
                                                                            Widget:=texture_block{DefaultDesiredSize:=vector2{X:=HorizontalLineSizeX,Y:=HorizontalLineSizeY},DefaultImage:=Images.Img_HorizontalLine}
                                                                        overlay_slot: 
                                                                            HorizontalAlignment:=horizontal_alignment.Center
                                                                            VerticalAlignment:=vertical_alignment.Center 
                                                                            Widget:=HT_TitleTextBlock
                                                                        overlay_slot: 
                                                                            HorizontalAlignment:=horizontal_alignment.Center
                                                                            VerticalAlignment:=vertical_alignment.Center 
                                                                            Widget:=texture_block{DefaultDesiredSize:=vector2{X:=10.0,Y:=LineSizeY},DefaultImage:=Images.EmptyTexture}
                                                            stack_box_slot: #Line 2: All titles
                                                                Padding:=margin{Left:=RowLeftPadding,Top:=RowTopPadding,Right:=RowRightPadding,Bottom:=RowBottomPadding}
                                                                    Widget:=overlay:
                                                                        Slots:=array:
                                                                            overlay_slot: 
                                                                                HorizontalAlignment:=horizontal_alignment.Fill
                                                                                VerticalAlignment:=vertical_alignment.Center
                                                                                Widget:=texture_block{DefaultDesiredSize:=vector2{X:=0.0,Y:=LineSizeY},DefaultImage:=Images.Img_LineBackground}
                                                                            overlay_slot: 
                                                                                Widget:=stack_box:
                                                                                    Orientation:=orientation.Horizontal
                                                                                        Slots:=array:
                                                                                            stack_box_slot: 
                                                                                                Padding:=margin{Left:=InterColumnLeftPadding,Top:=InterColumnTopPadding,Right:=InterColumnRightPadding,Bottom:=InterColumnBottomPadding}
                                                                                                Widget:=overlay:
                                                                                                    Slots:=array:
                                                                                                        overlay_slot:
                                                                                                            Widget:=text_block{DefaultText:=BasicMessage("Perk"), DefaultTextColor:=White, DefaultShadowOffset:=ShadowOffset, DefaultShadowOpacity := DefaultTextShadowOpacity}
                                                                                                        overlay_slot: 
                                                                                                            Widget:=texture_block{DefaultDesiredSize:=vector2{X:=BUColumn2SizeX,Y:=BUColumn2SizeY},DefaultImage:=Images.EmptyTexture}
                                                                                        
                                                                                            stack_box_slot: 
                                                                                                Padding:=margin{Left:=InterColumnLeftPadding,Top:=InterColumnTopPadding,Right:=InterColumnRightPadding,Bottom:=InterColumnBottomPadding}
                                                                                                Widget:=overlay:
                                                                                                    Slots:=array:
                                                                                                        overlay_slot:
                                                                                                            Widget:=text_block{DefaultText:=BasicMessage("Current"), DefaultTextColor:=White, DefaultShadowOffset:=ShadowOffset, DefaultShadowOpacity := DefaultTextShadowOpacity}
                                                                                                        overlay_slot: 
                                                                                                            Widget:=texture_block{DefaultDesiredSize:=vector2{X:=BUColumn4SizeX,Y:=BUColumn4SizeY},DefaultImage:=Images.EmptyTexture}
                                                                                            stack_box_slot: 
                                                                                                Padding:=margin{Left:=InterColumnLeftPadding,Top:=InterColumnTopPadding,Right:=InterColumnRightPadding,Bottom:=InterColumnBottomPadding}
                                                                                                Widget:=overlay:
                                                                                                    Slots:=array:
                                                                                                        overlay_slot:
                                                                                                            Widget:=text_block{DefaultText:=BasicMessage("Upgrade"), DefaultTextColor:=White, DefaultShadowOffset:=ShadowOffset, DefaultShadowOpacity := DefaultTextShadowOpacity}
                                                                                                        overlay_slot: 
                                                                                                            Widget:=texture_block{DefaultDesiredSize:=vector2{X:=BUColumn5SizeX,Y:=BUColumn5SizeY},DefaultImage:=Images.EmptyTexture}
                                                                                            stack_box_slot: 
                                                                                                Padding:=margin{Left:=InterColumnLeftPadding,Top:=InterColumnTopPadding,Right:=InterColumnRightPadding,Bottom:=InterColumnBottomPadding}
                                                                                                Widget:=overlay:
                                                                                                    Slots:=array:
                                                                                                        overlay_slot:
                                                                                                            Widget:=text_block{DefaultText:=BasicMessage("Upgrade cost"), DefaultTextColor:=White, DefaultShadowOffset:=ShadowOffset, DefaultShadowOpacity := DefaultTextShadowOpacity}
                                                                                                        overlay_slot: 
                                                                                                            Widget:=texture_block{DefaultDesiredSize:=vector2{X:=BUColumn6SizeX,Y:=BUColumn6SizeY},DefaultImage:=Images.EmptyTexture}
                                                            stack_box_slot: #Line 3: 
                                                                Padding:=margin{Left:=RowLeftPadding,Top:=RowTopPadding,Right:=RowRightPadding,Bottom:=RowBottomPadding}
                                                                    Widget:=overlay:
                                                                        Slots:=array:
                                                                            overlay_slot: 
                                                                                HorizontalAlignment:=horizontal_alignment.Fill
                                                                                VerticalAlignment:=vertical_alignment.Center
                                                                                Widget:=texture_block{DefaultDesiredSize:=vector2{X:=0.0,Y:=LineSizeY},DefaultImage:=Images.Img_LineBackground}
                                                                            overlay_slot: 
                                                                                Widget:=stack_box:
                                                                                    Orientation:=orientation.Horizontal
                                                                                        Slots:=array:
                                                                                            stack_box_slot: 
                                                                                                Padding:=margin{Left:=InterColumnLeftPadding,Top:=InterColumnTopPadding,Right:=InterColumnRightPadding,Bottom:=InterColumnBottomPadding}
                                                                                                Widget:=overlay:
                                                                                                    Slots:=array:
                                                                                                        overlay_slot:
                                                                                                            Widget:=text_block{DefaultText:=BasicMessage("Delay time"), DefaultTextColor:=White, DefaultShadowOffset:=ShadowOffset, DefaultShadowOpacity := DefaultTextShadowOpacity}
                                                                                                        overlay_slot: 
                                                                                                            Widget:=texture_block{DefaultDesiredSize:=vector2{X:=BUColumn2SizeX,Y:=BUColumn2SizeY},DefaultImage:=Images.EmptyTexture}
                                                                                            stack_box_slot: 
                                                                                                Padding:=margin{Left:=InterColumnLeftPadding,Top:=InterColumnTopPadding,Right:=InterColumnRightPadding,Bottom:=InterColumnBottomPadding}
                                                                                                Widget:=overlay:
                                                                                                    Slots:=array:
                                                                                                        overlay_slot:
                                                                                                            Widget:=HT_Row1_Current_TextBlock
                                                                                                        overlay_slot: 
                                                                                                            Widget:=texture_block{DefaultDesiredSize:=vector2{X:=BUColumn4SizeX,Y:=BUColumn4SizeY},DefaultImage:=Images.EmptyTexture}
                                                                                            stack_box_slot: 
                                                                                                Padding:=margin{Left:=InterColumnLeftPadding,Top:=InterColumnTopPadding,Right:=InterColumnRightPadding,Bottom:=InterColumnBottomPadding}
                                                                                                Widget:=overlay:
                                                                                                    Slots:=array:
                                                                                                        overlay_slot:
                                                                                                            Widget:=HT_Row1_Upgrade_TextBlock
                                                                                                        overlay_slot: 
                                                                                                            Widget:=texture_block{DefaultDesiredSize:=vector2{X:=BUColumn5SizeX,Y:=BUColumn5SizeY},DefaultImage:=Images.EmptyTexture}
                                                                                            stack_box_slot: 
                                                                                                Padding:=margin{Left:=InterColumnLeftPadding,Top:=InterColumnTopPadding,Right:=InterColumnRightPadding,Bottom:=InterColumnBottomPadding}
                                                                                                Widget:=overlay:
                                                                                                    Slots:=array:
                                                                                                        overlay_slot: 
                                                                                                            Widget:=texture_block{DefaultDesiredSize:=vector2{X:=BUColumn6SizeX,Y:=BUColumn6SizeY},DefaultImage:=Images.EmptyTexture}
                                                                                                        overlay_slot: 
                                                                                                            HorizontalAlignment:=horizontal_alignment.Left
                                                                                                            VerticalAlignment:=vertical_alignment.Center 
                                                                                                            Padding:=margin{Left:=65.0,Top:=0.0,Right:=0.0,Bottom:=0.0}
                                                                                                            Widget:=HT_Row1_UpgradeCost_TextBlock 
                                                                                                        overlay_slot: 
                                                                                                            HorizontalAlignment:=horizontal_alignment.Left
                                                                                                            VerticalAlignment:=vertical_alignment.Center 
                                                                                                            Padding:=margin{Left:=10.0,Top:=0.0,Right:=0.0,Bottom:=0.0}
                                                                                                            Widget:=texture_block{DefaultDesiredSize:=vector2{X:=IconSmallSizeX,Y:=IconSmallSizeY}, DefaultImage:=Images.Mutant_Material_Icon}
                                                                                                        overlay_slot:
                                                                                                            HorizontalAlignment:=horizontal_alignment.Fill
                                                                                                            VerticalAlignment:=vertical_alignment.Fill 
                                                                                                            Widget:=HT_Row1Button
                                                                                                        overlay_slot:
                                                                                                            Widget:=HT_Row1_UpgradeCost_TextBlock
                                                                                            

                                                            stack_box_slot: #Line 4: All titles
                                                                Padding:=margin{Left:=RowLeftPadding,Top:=RowTopPadding,Right:=RowRightPadding,Bottom:=RowBottomPadding}
                                                                    Widget:=overlay:
                                                                        Slots:=array:
                                                                            overlay_slot: 
                                                                                HorizontalAlignment:=horizontal_alignment.Fill
                                                                                VerticalAlignment:=vertical_alignment.Center
                                                                                Widget:=texture_block{DefaultDesiredSize:=vector2{X:=0.0,Y:=LineSizeY},DefaultImage:=Images.Img_LineBackground}
                                                                            overlay_slot: 
                                                                                Widget:=stack_box:
                                                                                    Orientation:=orientation.Horizontal
                                                                                        Slots:=array:
                                                                                            stack_box_slot: 
                                                                                                Padding:=margin{Left:=InterColumnLeftPadding,Top:=InterColumnTopPadding,Right:=InterColumnRightPadding,Bottom:=InterColumnBottomPadding}
                                                                                                Widget:=overlay:
                                                                                                    Slots:=array:
                                                                                                        overlay_slot:
                                                                                                            Widget:=text_block{DefaultText:=BasicMessage("Max accumulated tokens"), DefaultTextColor:=White, DefaultShadowOffset:=ShadowOffset, DefaultShadowOpacity := DefaultTextShadowOpacity}
                                                                                                        overlay_slot: 
                                                                                                            Widget:=texture_block{DefaultDesiredSize:=vector2{X:=BUColumn2SizeX,Y:=BUColumn2SizeY},DefaultImage:=Images.EmptyTexture}
                                                                                            
                                                                                            stack_box_slot: 
                                                                                                Padding:=margin{Left:=InterColumnLeftPadding,Top:=InterColumnTopPadding,Right:=InterColumnRightPadding,Bottom:=InterColumnBottomPadding}
                                                                                                Widget:=overlay:
                                                                                                    Slots:=array:
                                                                                                        overlay_slot:
                                                                                                            Widget:=HT_Row2_Current_TextBlock
                                                                                                        overlay_slot: 
                                                                                                            Widget:=texture_block{DefaultDesiredSize:=vector2{X:=BUColumn4SizeX,Y:=BUColumn4SizeY},DefaultImage:=Images.EmptyTexture}
                                                                                            stack_box_slot: 
                                                                                                Padding:=margin{Left:=InterColumnLeftPadding,Top:=InterColumnTopPadding,Right:=InterColumnRightPadding,Bottom:=InterColumnBottomPadding}
                                                                                                Widget:=overlay:
                                                                                                    Slots:=array:
                                                                                                        overlay_slot:
                                                                                                            Widget:=HT_Row2_Upgrade_TextBlock
                                                                                                        overlay_slot: 
                                                                                                            Widget:=texture_block{DefaultDesiredSize:=vector2{X:=BUColumn5SizeX,Y:=BUColumn5SizeY},DefaultImage:=Images.EmptyTexture}
                                                                                            stack_box_slot: 
                                                                                                Padding:=margin{Left:=InterColumnLeftPadding,Top:=InterColumnTopPadding,Right:=InterColumnRightPadding,Bottom:=InterColumnBottomPadding}
                                                                                                Widget:=overlay:
                                                                                                    Slots:=array:
                                                                                                        overlay_slot: 
                                                                                                            Widget:=texture_block{DefaultDesiredSize:=vector2{X:=BUColumn6SizeX,Y:=BUColumn6SizeY},DefaultImage:=Images.EmptyTexture}
                                                                                                        overlay_slot: 
                                                                                                            HorizontalAlignment:=horizontal_alignment.Left
                                                                                                            VerticalAlignment:=vertical_alignment.Center 
                                                                                                            Padding:=margin{Left:=65.0,Top:=0.0,Right:=0.0,Bottom:=0.0}
                                                                                                            Widget:=HT_Row2_UpgradeCost_TextBlock 
                                                                                                        overlay_slot: 
                                                                                                            HorizontalAlignment:=horizontal_alignment.Left
                                                                                                            VerticalAlignment:=vertical_alignment.Center 
                                                                                                            Padding:=margin{Left:=10.0,Top:=0.0,Right:=0.0,Bottom:=0.0}
                                                                                                            Widget:=texture_block{DefaultDesiredSize:=vector2{X:=IconSmallSizeX,Y:=IconSmallSizeY}, DefaultImage:=Images.AgedSaphire_Icon}
                                                                                                        overlay_slot:
                                                                                                            HorizontalAlignment:=horizontal_alignment.Fill
                                                                                                            VerticalAlignment:=vertical_alignment.Fill 
                                                                                                            Widget:=HT_Row2Button
                                                                                                        overlay_slot:
                                                                                                            Widget:=HT_Row2_UpgradeCost_TextBlock   
                                                            stack_box_slot: #Line 9: Close button
                                                                Padding:=margin{Left:=RowLeftPadding,Top:=RowTopPadding,Right:=RowRightPadding,Bottom:=RowBottomPadding}
                                                                    Widget:=overlay:
                                                                        Slots:=array:
                                                                            overlay_slot: 
                                                                                HorizontalAlignment:=horizontal_alignment.Center
                                                                                VerticalAlignment:=vertical_alignment.Center
                                                                                Widget:=stack_box:
                                                                                    Orientation:=orientation.Horizontal
                                                                                        Slots:=array:
                                                                                            stack_box_slot: 
                                                                                                Padding:=margin{Left:=InterColumnLeftPadding,Top:=InterColumnTopPadding,Right:=InterColumnRightPadding,Bottom:=InterColumnBottomPadding}
                                                                                                Widget:=overlay:
                                                                                                    Slots:=array:
                                                                                                        overlay_slot: 
                                                                                                            Widget:=texture_block{DefaultDesiredSize:=vector2{X:=BigCloseButtonSizeX,Y:=BigCloseButtonSizeY},DefaultImage:=Images.EmptyTexture}
                                                                                                        overlay_slot: 
                                                                                                            HorizontalAlignment:=horizontal_alignment.Fill
                                                                                                            VerticalAlignment:=vertical_alignment.Fill
                                                                                                            Padding:=margin{Left:=InterColumnLeftPadding,Top:=InterColumnTopPadding,Right:=InterColumnRightPadding,Bottom:=InterColumnBottomPadding}
                                                                                                            Widget:=HT_CloseButton  

    MakeDungeonPerksCanvas(): canvas=
        Canvas := canvas:
            Slots:=array:
                #Black background image
                canvas_slot:
                    Anchors:=anchors{Minimum:=vector2{X:=0.5,Y:=0.5},Maximum:=vector2{X:=0.5,Y:=0.5}}
                    ZOrder:=0
                    SizeToContent:=true
                    Offsets:=margin{Left:=0.0,Top:=0.0,Right:=0.0,Bottom:=0.0} #PositionX, PositionY, SizeX, SizeY
                    Alignment:=vector2{X:=0.5,Y:=0.5}
                    Widget:=texture_block{DefaultImage:=Images.BlackBackground,DefaultDesiredSize:=vector2{X:=2100.0,Y:=1500.0}}
                #Background image
                canvas_slot:
                    Anchors:=anchors{Minimum:=vector2{X:=0.5,Y:=0.5},Maximum:=vector2{X:=0.5,Y:=0.5}}
                    ZOrder:=0
                    Offsets:=margin{Left:=0.0,Top:=0.0,Right:=BackgroundImageSizeX,Bottom:=BackgroundImageSizeY} #PositionX, PositionY, SizeX, SizeY
                    Alignment:=vector2{X:=0.5,Y:=0.5}
                    Widget:=texture_block:
                        DefaultImage:=Images.Img_MainBackground
                        DefaultDesiredSize:=vector2{X:=BackgroundImageSizeX,Y:=BackgroundImageSizeY}
                #Main overlay
                canvas_slot:
                    Anchors:=anchors{Minimum:=vector2{X:=0.5,Y:=0.5},Maximum:=vector2{X:=0.5,Y:=0.5}}
                    ZOrder:=0
                    Alignment:=vector2{X:=0.5,Y:=0.5}
                    Widget:=overlay:
                        Slots:=array:
                            overlay_slot:
                                HorizontalAlignment:=horizontal_alignment.Center
                                VerticalAlignment:=vertical_alignment.Center
                                Padding:=margin{Left:=0.0,Top:=10.0,Right:=0.0,Bottom:=0.0}
                                Widget:=stack_box:
                                    Orientation:=orientation.Vertical
                                        Slots:=array:
                                            stack_box_slot: #Sapphires per produc
                                                HorizontalAlignment:=horizontal_alignment.Right
                                                Widget:=stack_box:
                                                    Orientation:=orientation.Horizontal
                                                        Slots:=array:
                                                            stack_box_slot: #Current MUTANTMATERIAL
                                                                Padding:=margin{Left:=InterColumnLeftPadding,Top:=InterColumnTopPadding,Right:=InterColumnRightPadding,Bottom:=InterColumnBottomPadding}
                                                                Widget:=overlay:
                                                                    Slots:=array:
                                                                        overlay_slot:
                                                                            Widget:= texture_block{DefaultDesiredSize:=vector2{X:=SizeBoxSaphiresX,Y:=SizeBoxSaphiresY}, DefaultImage:=Images.Img_LineBackground2}
                                                                        overlay_slot:
                                                                            HorizontalAlignment:=horizontal_alignment.Left
                                                                            VerticalAlignment:=vertical_alignment.Center 
                                                                            Padding:=margin{Left:=10.0,Top:=0.0,Right:=0.0,Bottom:=0.0}
                                                                            Widget:=texture_block{DefaultDesiredSize:=vector2{X:=IconSmallSizeX,Y:=IconSmallSizeY}, DefaultImage:=Images.Mutant_Material_Icon}
                                                                        overlay_slot: 
                                                                            HorizontalAlignment:=horizontal_alignment.Left
                                                                            VerticalAlignment:=vertical_alignment.Center 
                                                                            Padding:=margin{Left:=65.0,Top:=0.0,Right:=0.0,Bottom:=0.0}
                                                                            Widget:=DT_TotalMutantMaterial_TextBlock
                                                            stack_box_slot: #Current AGEDSAPPHIRES
                                                                Padding:=margin{Left:=InterColumnLeftPadding,Top:=InterColumnTopPadding,Right:=InterColumnRightPadding,Bottom:=InterColumnBottomPadding}
                                                                Widget:=overlay:
                                                                    Slots:=array:
                                                                        overlay_slot:
                                                                            Widget:= texture_block{DefaultDesiredSize:=vector2{X:=SizeBoxSaphiresX,Y:=SizeBoxSaphiresY}, DefaultImage:=Images.Img_LineBackground2}
                                                                        overlay_slot:
                                                                            HorizontalAlignment:=horizontal_alignment.Left
                                                                            VerticalAlignment:=vertical_alignment.Center 
                                                                            Padding:=margin{Left:=10.0,Top:=0.0,Right:=0.0,Bottom:=0.0}
                                                                            Widget:=texture_block{DefaultDesiredSize:=vector2{X:=IconSmallSizeX,Y:=IconSmallSizeY}, DefaultImage:=Images.AgedSaphire_Icon}
                                                                        overlay_slot: 
                                                                            HorizontalAlignment:=horizontal_alignment.Left
                                                                            VerticalAlignment:=vertical_alignment.Center 
                                                                            Padding:=margin{Left:=65.0,Top:=0.0,Right:=0.0,Bottom:=0.0}
                                                                            Widget:=DT_TotalAgedSapphires_TextBlock       
                                            stack_box_slot: #Perk upgrades
                                                Widget:=stack_box:
                                                    Orientation:=orientation.Vertical
                                                        Slots:=array:
                                                            stack_box_slot: #Line 0: All titles
                                                                HorizontalAlignment:=horizontal_alignment.Right
                                                                Padding:=margin{Left:=RowLeftPadding,Top:=25.0,Right:=RowRightPadding,Bottom:=RowBottomPadding}
                                                                    Widget:=overlay:
                                                                        Slots:=array:
                                                                            overlay_slot: 
                                                                                HorizontalAlignment:=horizontal_alignment.Fill
                                                                                VerticalAlignment:=vertical_alignment.Center
                                                                                Widget:=texture_block{DefaultDesiredSize:=vector2{X:=0.0,Y:=LineSizeY},DefaultImage:=Images.Img_LineBackground}
                                                            stack_box_slot: #Line 1: Title
                                                                Padding:=margin{Left:=RowLeftPadding,Top:=RowTopPadding,Right:=RowRightPadding,Bottom:=RowBottomPadding}
                                                                HorizontalAlignment:=horizontal_alignment.Fill
                                                                VerticalAlignment:=vertical_alignment.Bottom
                                                                Widget:=overlay:
                                                                    Slots:=array:
                                                                        overlay_slot: 
                                                                            HorizontalAlignment:=horizontal_alignment.Fill
                                                                            VerticalAlignment:=vertical_alignment.Bottom
                                                                            Widget:=texture_block{DefaultDesiredSize:=vector2{X:=HorizontalLineSizeX,Y:=HorizontalLineSizeY},DefaultImage:=Images.Img_HorizontalLine}
                                                                        overlay_slot: 
                                                                            HorizontalAlignment:=horizontal_alignment.Center
                                                                            VerticalAlignment:=vertical_alignment.Center 
                                                                            Widget:=DT_TitleTextBlock
                                                                        overlay_slot: 
                                                                            HorizontalAlignment:=horizontal_alignment.Center
                                                                            VerticalAlignment:=vertical_alignment.Center 
                                                                            Widget:=texture_block{DefaultDesiredSize:=vector2{X:=10.0,Y:=LineSizeY},DefaultImage:=Images.EmptyTexture}
                                                            stack_box_slot: #Line 2: All titles
                                                                Padding:=margin{Left:=RowLeftPadding,Top:=RowTopPadding,Right:=RowRightPadding,Bottom:=RowBottomPadding}
                                                                    Widget:=overlay:
                                                                        Slots:=array:
                                                                            overlay_slot: 
                                                                                HorizontalAlignment:=horizontal_alignment.Fill
                                                                                VerticalAlignment:=vertical_alignment.Center
                                                                                Widget:=texture_block{DefaultDesiredSize:=vector2{X:=0.0,Y:=LineSizeY},DefaultImage:=Images.Img_LineBackground}
                                                                            overlay_slot: 
                                                                                Widget:=stack_box:
                                                                                    Orientation:=orientation.Horizontal
                                                                                        Slots:=array:
                                                                                            stack_box_slot: 
                                                                                                Padding:=margin{Left:=InterColumnLeftPadding,Top:=InterColumnTopPadding,Right:=InterColumnRightPadding,Bottom:=InterColumnBottomPadding}
                                                                                                Widget:=overlay:
                                                                                                    Slots:=array:
                                                                                                        overlay_slot:
                                                                                                            Widget:=text_block{DefaultText:=BasicMessage("Perk"), DefaultTextColor:=White, DefaultShadowOffset:=ShadowOffset, DefaultShadowOpacity := DefaultTextShadowOpacity}
                                                                                                        overlay_slot: 
                                                                                                            Widget:=texture_block{DefaultDesiredSize:=vector2{X:=BUColumn2SizeX,Y:=BUColumn2SizeY},DefaultImage:=Images.EmptyTexture}
                                                                                        
                                                                                            stack_box_slot: 
                                                                                                Padding:=margin{Left:=InterColumnLeftPadding,Top:=InterColumnTopPadding,Right:=InterColumnRightPadding,Bottom:=InterColumnBottomPadding}
                                                                                                Widget:=overlay:
                                                                                                    Slots:=array:
                                                                                                        overlay_slot:
                                                                                                            Widget:=text_block{DefaultText:=BasicMessage("Current"), DefaultTextColor:=White, DefaultShadowOffset:=ShadowOffset, DefaultShadowOpacity := DefaultTextShadowOpacity}
                                                                                                        overlay_slot: 
                                                                                                            Widget:=texture_block{DefaultDesiredSize:=vector2{X:=BUColumn4SizeX,Y:=BUColumn4SizeY},DefaultImage:=Images.EmptyTexture}
                                                                                            stack_box_slot: 
                                                                                                Padding:=margin{Left:=InterColumnLeftPadding,Top:=InterColumnTopPadding,Right:=InterColumnRightPadding,Bottom:=InterColumnBottomPadding}
                                                                                                Widget:=overlay:
                                                                                                    Slots:=array:
                                                                                                        overlay_slot:
                                                                                                            Widget:=text_block{DefaultText:=BasicMessage("Upgrade"), DefaultTextColor:=White, DefaultShadowOffset:=ShadowOffset, DefaultShadowOpacity := DefaultTextShadowOpacity}
                                                                                                        overlay_slot: 
                                                                                                            Widget:=texture_block{DefaultDesiredSize:=vector2{X:=BUColumn5SizeX,Y:=BUColumn5SizeY},DefaultImage:=Images.EmptyTexture}
                                                                                            stack_box_slot: 
                                                                                                Padding:=margin{Left:=InterColumnLeftPadding,Top:=InterColumnTopPadding,Right:=InterColumnRightPadding,Bottom:=InterColumnBottomPadding}
                                                                                                Widget:=overlay:
                                                                                                    Slots:=array:
                                                                                                        overlay_slot:
                                                                                                            Widget:=text_block{DefaultText:=BasicMessage("Upgrade cost"), DefaultTextColor:=White, DefaultShadowOffset:=ShadowOffset, DefaultShadowOpacity := DefaultTextShadowOpacity}
                                                                                                        overlay_slot: 
                                                                                                            Widget:=texture_block{DefaultDesiredSize:=vector2{X:=BUColumn6SizeX,Y:=BUColumn6SizeY},DefaultImage:=Images.EmptyTexture}
                                                            stack_box_slot: #Line 3: 
                                                                Padding:=margin{Left:=RowLeftPadding,Top:=RowTopPadding,Right:=RowRightPadding,Bottom:=RowBottomPadding}
                                                                    Widget:=overlay:
                                                                        Slots:=array:
                                                                            overlay_slot: 
                                                                                HorizontalAlignment:=horizontal_alignment.Fill
                                                                                VerticalAlignment:=vertical_alignment.Center
                                                                                Widget:=texture_block{DefaultDesiredSize:=vector2{X:=0.0,Y:=LineSizeY},DefaultImage:=Images.Img_LineBackground}
                                                                            overlay_slot: 
                                                                                Widget:=stack_box:
                                                                                    Orientation:=orientation.Horizontal
                                                                                        Slots:=array:
                                                                                            stack_box_slot: 
                                                                                                Padding:=margin{Left:=InterColumnLeftPadding,Top:=InterColumnTopPadding,Right:=InterColumnRightPadding,Bottom:=InterColumnBottomPadding}
                                                                                                Widget:=overlay:
                                                                                                    Slots:=array:
                                                                                                        overlay_slot:
                                                                                                            Widget:=text_block{DefaultText:=BasicMessage("Delay time"), DefaultTextColor:=White, DefaultShadowOffset:=ShadowOffset, DefaultShadowOpacity := DefaultTextShadowOpacity}
                                                                                                        overlay_slot: 
                                                                                                            Widget:=texture_block{DefaultDesiredSize:=vector2{X:=BUColumn2SizeX,Y:=BUColumn2SizeY},DefaultImage:=Images.EmptyTexture}
                                                                                            stack_box_slot: 
                                                                                                Padding:=margin{Left:=InterColumnLeftPadding,Top:=InterColumnTopPadding,Right:=InterColumnRightPadding,Bottom:=InterColumnBottomPadding}
                                                                                                Widget:=overlay:
                                                                                                    Slots:=array:
                                                                                                        overlay_slot:
                                                                                                            Widget:=DT_Row1_Current_TextBlock
                                                                                                        overlay_slot: 
                                                                                                            Widget:=texture_block{DefaultDesiredSize:=vector2{X:=BUColumn4SizeX,Y:=BUColumn4SizeY},DefaultImage:=Images.EmptyTexture}
                                                                                            stack_box_slot: 
                                                                                                Padding:=margin{Left:=InterColumnLeftPadding,Top:=InterColumnTopPadding,Right:=InterColumnRightPadding,Bottom:=InterColumnBottomPadding}
                                                                                                Widget:=overlay:
                                                                                                    Slots:=array:
                                                                                                        overlay_slot:
                                                                                                            Widget:=DT_Row1_Upgrade_TextBlock
                                                                                                        overlay_slot: 
                                                                                                            Widget:=texture_block{DefaultDesiredSize:=vector2{X:=BUColumn5SizeX,Y:=BUColumn5SizeY},DefaultImage:=Images.EmptyTexture}
                                                                                            stack_box_slot: 
                                                                                                Padding:=margin{Left:=InterColumnLeftPadding,Top:=InterColumnTopPadding,Right:=InterColumnRightPadding,Bottom:=InterColumnBottomPadding}
                                                                                                Widget:=overlay:
                                                                                                    Slots:=array:
                                                                                                        overlay_slot: 
                                                                                                            Widget:=texture_block{DefaultDesiredSize:=vector2{X:=BUColumn6SizeX,Y:=BUColumn6SizeY},DefaultImage:=Images.EmptyTexture}
                                                                                                        overlay_slot: 
                                                                                                            HorizontalAlignment:=horizontal_alignment.Left
                                                                                                            VerticalAlignment:=vertical_alignment.Center 
                                                                                                            Padding:=margin{Left:=65.0,Top:=0.0,Right:=0.0,Bottom:=0.0}
                                                                                                            Widget:=DT_Row1_UpgradeCost_TextBlock 
                                                                                                        overlay_slot: 
                                                                                                            HorizontalAlignment:=horizontal_alignment.Left
                                                                                                            VerticalAlignment:=vertical_alignment.Center 
                                                                                                            Padding:=margin{Left:=10.0,Top:=0.0,Right:=0.0,Bottom:=0.0}
                                                                                                            Widget:=texture_block{DefaultDesiredSize:=vector2{X:=IconSmallSizeX,Y:=IconSmallSizeY}, DefaultImage:=Images.Mutant_Material_Icon}
                                                                                                        overlay_slot:
                                                                                                            HorizontalAlignment:=horizontal_alignment.Fill
                                                                                                            VerticalAlignment:=vertical_alignment.Fill  
                                                                                                            Widget:=DT_Row1Button
                                                                                                        overlay_slot:
                                                                                                            Widget:=DT_Row1_UpgradeCost_TextBlock
                                                                                            

                                                            stack_box_slot: #Line 4: All titles
                                                                Padding:=margin{Left:=RowLeftPadding,Top:=RowTopPadding,Right:=RowRightPadding,Bottom:=RowBottomPadding}
                                                                    Widget:=overlay:
                                                                        Slots:=array:
                                                                            overlay_slot: 
                                                                                HorizontalAlignment:=horizontal_alignment.Fill
                                                                                VerticalAlignment:=vertical_alignment.Center
                                                                                Widget:=texture_block{DefaultDesiredSize:=vector2{X:=0.0,Y:=LineSizeY},DefaultImage:=Images.Img_LineBackground}
                                                                            overlay_slot: 
                                                                                Widget:=stack_box:
                                                                                    Orientation:=orientation.Horizontal
                                                                                        Slots:=array:
                                                                                            stack_box_slot: 
                                                                                                Padding:=margin{Left:=InterColumnLeftPadding,Top:=InterColumnTopPadding,Right:=InterColumnRightPadding,Bottom:=InterColumnBottomPadding}
                                                                                                Widget:=overlay:
                                                                                                    Slots:=array:
                                                                                                        overlay_slot:
                                                                                                            Widget:=text_block{DefaultText:=BasicMessage("Max accumulated tokens"), DefaultTextColor:=White, DefaultShadowOffset:=ShadowOffset, DefaultShadowOpacity := DefaultTextShadowOpacity}
                                                                                                        overlay_slot: 
                                                                                                            Widget:=texture_block{DefaultDesiredSize:=vector2{X:=BUColumn2SizeX,Y:=BUColumn2SizeY},DefaultImage:=Images.EmptyTexture}
                                                                                            
                                                                                            stack_box_slot: 
                                                                                                Padding:=margin{Left:=InterColumnLeftPadding,Top:=InterColumnTopPadding,Right:=InterColumnRightPadding,Bottom:=InterColumnBottomPadding}
                                                                                                Widget:=overlay:
                                                                                                    Slots:=array:
                                                                                                        overlay_slot:
                                                                                                            Widget:=DT_Row2_Current_TextBlock
                                                                                                        overlay_slot: 
                                                                                                            Widget:=texture_block{DefaultDesiredSize:=vector2{X:=BUColumn4SizeX,Y:=BUColumn4SizeY},DefaultImage:=Images.EmptyTexture}
                                                                                            stack_box_slot: 
                                                                                                Padding:=margin{Left:=InterColumnLeftPadding,Top:=InterColumnTopPadding,Right:=InterColumnRightPadding,Bottom:=InterColumnBottomPadding}
                                                                                                Widget:=overlay:
                                                                                                    Slots:=array:
                                                                                                        overlay_slot:
                                                                                                            Widget:=DT_Row2_Upgrade_TextBlock
                                                                                                        overlay_slot: 
                                                                                                            Widget:=texture_block{DefaultDesiredSize:=vector2{X:=BUColumn5SizeX,Y:=BUColumn5SizeY},DefaultImage:=Images.EmptyTexture}
                                                                                            stack_box_slot: 
                                                                                                Padding:=margin{Left:=InterColumnLeftPadding,Top:=InterColumnTopPadding,Right:=InterColumnRightPadding,Bottom:=InterColumnBottomPadding}
                                                                                                Widget:=overlay:
                                                                                                    Slots:=array:
                                                                                                        overlay_slot: 
                                                                                                            Widget:=texture_block{DefaultDesiredSize:=vector2{X:=BUColumn6SizeX,Y:=BUColumn6SizeY},DefaultImage:=Images.EmptyTexture}
                                                                                                        overlay_slot: 
                                                                                                            HorizontalAlignment:=horizontal_alignment.Left
                                                                                                            VerticalAlignment:=vertical_alignment.Center 
                                                                                                            Padding:=margin{Left:=65.0,Top:=0.0,Right:=0.0,Bottom:=0.0}
                                                                                                            Widget:=DT_Row2_UpgradeCost_TextBlock 
                                                                                                        overlay_slot: 
                                                                                                            HorizontalAlignment:=horizontal_alignment.Left
                                                                                                            VerticalAlignment:=vertical_alignment.Center 
                                                                                                            Padding:=margin{Left:=10.0,Top:=0.0,Right:=0.0,Bottom:=0.0}
                                                                                                            Widget:=texture_block{DefaultDesiredSize:=vector2{X:=IconSmallSizeX,Y:=IconSmallSizeY}, DefaultImage:=Images.AgedSaphire_Icon}
                                                                                                        overlay_slot:
                                                                                                            HorizontalAlignment:=horizontal_alignment.Fill
                                                                                                            VerticalAlignment:=vertical_alignment.Fill  
                                                                                                            Widget:=DT_Row2Button
                                                                                                        overlay_slot:
                                                                                                            Widget:=DT_Row2_UpgradeCost_TextBlock   
                                                            stack_box_slot: #Line 9: Close button
                                                                Padding:=margin{Left:=RowLeftPadding,Top:=RowTopPadding,Right:=RowRightPadding,Bottom:=RowBottomPadding}
                                                                    Widget:=overlay:
                                                                        Slots:=array:
                                                                            overlay_slot: 
                                                                                HorizontalAlignment:=horizontal_alignment.Center
                                                                                VerticalAlignment:=vertical_alignment.Center
                                                                                Widget:=stack_box:
                                                                                    Orientation:=orientation.Horizontal
                                                                                        Slots:=array:
                                                                                            stack_box_slot: 
                                                                                                Padding:=margin{Left:=InterColumnLeftPadding,Top:=InterColumnTopPadding,Right:=InterColumnRightPadding,Bottom:=InterColumnBottomPadding}
                                                                                                Widget:=overlay:
                                                                                                    Slots:=array:
                                                                                                        overlay_slot: 
                                                                                                            Widget:=texture_block{DefaultDesiredSize:=vector2{X:=BigCloseButtonSizeX,Y:=BigCloseButtonSizeY},DefaultImage:=Images.EmptyTexture}
                                                                                                        overlay_slot: 
                                                                                                            HorizontalAlignment:=horizontal_alignment.Fill
                                                                                                            VerticalAlignment:=vertical_alignment.Fill
                                                                                                            Padding:=margin{Left:=InterColumnLeftPadding,Top:=InterColumnTopPadding,Right:=InterColumnRightPadding,Bottom:=InterColumnBottomPadding}
                                                                                                            Widget:=DT_CloseButton                                  
MakeTokensUI<constructor>(
    InBakeryMachinesManager : bakery_machines_manager)<transacts> := tokens_ui :
        BakeryMachinesManager := InBakeryMachinesManager
                                                                                                               