
using { /Fortnite.com/Devices }
using { /Verse.org/Simulation }
using { /UnrealEngine.com/Temporary/Diagnostics }
using { /Verse.org/Assets }
using { /UnrealEngine.com/Temporary/SpatialMath }
using { /Fortnite.com/Characters }

# See https://dev.epicgames.com/documentation/en-us/uefn/create-your-own-device-in-verse for how to create a verse device.


othermessages_module<public> := module:
    GoodLuck<public><localizes>():message = "Good luck!"
    Recomendation<public><localizes>():message = "Don't skip the tutorial!"
    Title<public><localizes>():message = "Tutorial"
    Funnel<public><localizes>():message = "Use the funnel"
    GatherWater<public><localizes>():message = "Gather 5 Water"
    GatherFlour<public><localizes>():message = "Gather 5 Flour"
    CheckPick<public><localizes>():message = "Check the Pick"
    GoodJob<public><localizes>():message = "Good job!"
    OpenValves<public><localizes>():message = "Open the valves"
    UseCoal<public><localizes>():message = "Add 30 Coal"
    SendDrone<public><localizes>():message = "Send the drone"
    CanNotDrone<public><localizes>():message = "Can't send the drone now"
    UpgradePick<public><localizes>():message = "Upgrade the Pick"
    CheckComputer<public><localizes>():message = "Check the computer"
    BakeBread<public><localizes>():message = "Bake Bread"
    ImproveProduction<public><localizes>():message = "Boost Baker's production"
    Problem<public><localizes>():message = "Let's fix it and continue!"
    ThisTheBakery<public><localizes>():message = "This is your Bakery"
    CompleteAgain<public><localizes>():message = "Redo tutorial if needed"
    Shop<public><localizes>():message = "Buy upgrades in the shop"
    Pets<public><localizes>():message = "Complete Quests to unlock Sourdoughs (Pets)"
<#
tutorial_module<public> := module:
    Phase0_UKMessage0<public><localizes>()<transacts>:message = "Welcome to Breaking Bread! Let's start..."
    Phase0_UKMessage1<public><localizes>()<transacts>:message = "Slate Bread may give key ingredients"

    Phase1_UKMessage0<public><localizes>():message = "Add ingredients to the funnel"
    Phase1_UKMessage1<public><localizes>():message = "Use the funnel"
    Phase1_UKMessage2<public><localizes>():message = "Watch your deposits fill up"

    Phase2_UKMessage0<public><localizes>():message = "Now, let's learn about Bakers"
    Phase2_UKMessage1<public><localizes>():message = "Gather 5 Water"
    Phase2_UKMessage2<public><localizes>():message = "Gather 5 Flour"
    Phase2_UKMessage3<public><localizes>():message = "Add ingredients to the funnel"

    Phase2_UKMessage4<public><localizes>():message = "Open valves to send ingredients"
    Phase2_UKMessage44<public><localizes>():message = "Green valves > Yellow valves"
    Phase2_UKMessage5<public><localizes>():message = "Keep bars full for the next batch"
    Phase2_UKMessage6<public><localizes>():message = "Low on ingredients? Gather more!"

    Phase2_UKMessage7<public><localizes>():message = "Good job! The Baker is working"
    Phase2_UKMessage8<public><localizes>():message = "Green lights = Baking in progress"
    Phase2_UKMessage9<public><localizes>():message = "Speeding up the Baker for now"
    Phase2_UKMessage10<public><localizes>():message = "Wow! That looks great!"
    Phase2_UKMessage11<public><localizes>():message = "Send the Bread with the drone"

    Phase3_UKMessage0<public><localizes>():message = "Upgrade the Pick Gather"
    Phase3_UKMessage1<public><localizes>():message = "You gather more per pick now!"
    Phase3_UKMessage2<public><localizes>():message = "Check Pick damage on Slate Bread"
    Phase3_UKMessage3<public><localizes>():message = "Cool!"
    Phase3_UKMessage4<public><localizes>():message = "Bars show needed resources"
    Phase3_UKMessage5<public><localizes>():message = "Check Baker resources on PC"
    Phase3_UKMessage6<public><localizes>():message = "Interact with it!"
    Phase3_UKMessage7<public><localizes>():message = "See resource amounts?"
    Phase3_UKMessage8<public><localizes>():message = "Improve Baker's production"
    Phase3_UKMessage9<public><localizes>():message = "Upgrading raises needs"

    Phase4_UKMessage0<public><localizes>():message = "Use Coal to boost production"
    Phase4_UKMessage1<public><localizes>():message = "Add 30 Coal to the funnel"
    Phase4_UKMessage2<public><localizes>():message = "Creates Super Energy for a while"
    Phase4_UKMessage3<public><localizes>():message = "Check remaining Super Energy"
    Phase4_UKMessage4<public><localizes>():message = "See your multiplier here"
    Phase4_UKMessage5<public><localizes>():message = "Coal gives Shared Super Energy"
    Phase4_UKMessage6<public><localizes>():message = "Boosts all players slightly"
    Phase4_UKMessage7<public><localizes>():message = "If all have it, you get a boost!"
    Phase4_UKMessage8<public><localizes>():message = "Time to apply what you learned!"

    Phase4_UKMessage10<public><localizes>():message = "Add ingredients to the funnel"
    Phase4_UKMessage11<public><localizes>():message = "Add Coal before opening valves"
    Phase4_UKMessage12<public><localizes>():message = "Open valves to send ingredients"
    Phase4_UKMessage13<public><localizes>():message = "Check needs & Bake Bread!"
    Phase4_UKMessage14<public><localizes>():message = "Good job! Baker is baking"
    Phase4_UKMessage15<public><localizes>():message = "Green lights = Working"
    Phase4_UKMessage16<public><localizes>():message = "See how needs changed?"
    Phase4_UKMessage17<public><localizes>():message = "Next batch is prepared"
    Phase4_UKMessage19<public><localizes>():message = "That looks great!"
    Phase4_UKMessage20<public><localizes>():message = "Send the Bread with the drone"
    Phase4_UKMessage21<public><localizes>():message = "Tutorial done! Back to Bakery"

    Extra_Message0<public><localizes>():message = "Pick already upgraded"
    Extra_Message1<public><localizes>():message = "Baker already upgraded"
#>
tutorial_en_module<public> := module:
    Phase0_UKMessage0<public>:string = "Welcome to Breaking Bread! Let's start..."
    Phase0_UKMessage1<public>:string = "Stale Bread may give key ingredients"
    
    Phase1_UKMessage0<public>:string = "Add ingredients to the funnel"
    Phase1_UKMessage1<public>:string = "Use the funnel"
    Phase1_UKMessage2<public>:string = "Watch your deposits fill up"
    
    Phase2_UKMessage0<public>:string = "Now, let's learn about Bakers"
    Phase2_UKMessage1<public>:string = "Gather 5 Water"
    Phase2_UKMessage2<public>:string = "Gather 5 Flour"
    Phase2_UKMessage3<public>:string = "Add ingredients to the funnel"
    
    Phase2_UKMessage4<public>:string = "Open valves to send ingredients"
    Phase2_UKMessage44<public>:string = "Green valves > Yellow valves"
    Phase2_UKMessage5<public>:string = "Keep bars full for the next batch"
    Phase2_UKMessage6<public>:string = "Low on ingredients? Gather more!"
    
    Phase2_UKMessage7<public>:string = "Good job! The Baker is working"
    Phase2_UKMessage8<public>:string = "Green lights = Baking in progress"
    Phase2_UKMessage9<public>:string = "Speeding up the Baker for now"
    Phase2_UKMessage10<public>:string = "Wow! That looks great!"
    Phase2_UKMessage11<public>:string = "Send the Bread with the drone"
    
    Phase3_UKMessage0<public>:string = "Upgrade the Pick Gather"
    Phase3_UKMessage1<public>:string = "You gather more per pick now!"
    Phase3_UKMessage2<public>:string = "Check Pick damage on Stale Bread"
    Phase3_UKMessage3<public>:string = "Cool!"
    Phase3_UKMessage4<public>:string = "Bars show needed resources"
    Phase3_UKMessage5<public>:string = "Check Baker resources on PC"
    Phase3_UKMessage6<public>:string = "Interact with it!"
    Phase3_UKMessage7<public>:string = "See resource amounts?"
    Phase3_UKMessage8<public>:string = "Improve Baker's production"
    Phase3_UKMessage9<public>:string = "Upgrading raises needs"
    
    Phase4_UKMessage0<public>:string = "Use Coal to boost production"
    Phase4_UKMessage1<public>:string = "Add 30 Coal to the funnel"
    Phase4_UKMessage2<public>:string = "Creates Super Energy for a while"
    Phase4_UKMessage3<public>:string = "Check remaining Super Energy"
    Phase4_UKMessage4<public>:string = "See your multiplier here"
    Phase4_UKMessage5<public>:string = "Coal gives Shared Super Energy"
    Phase4_UKMessage6<public>:string = "Boosts all players slightly"
    Phase4_UKMessage7<public>:string = "If all have it, you get a boost!"
    Phase4_UKMessage8<public>:string = "Time to apply what you learned!"
    
    Phase4_UKMessage10<public>:string = "Add ingredients to the funnel"
    Phase4_UKMessage11<public>:string = "Add Coal before opening valves"
    Phase4_UKMessage12<public>:string = "Open valves to send ingredients"
    Phase4_UKMessage13<public>:string = "Add missing ingredients to the funnel & Bake Bread!"
    Phase4_UKMessage14<public>:string = "Good job! Baker is baking"
    Phase4_UKMessage15<public>:string = "Green lights = Working"
    Phase4_UKMessage16<public>:string = "See how needs changed?"
    Phase4_UKMessage17<public>:string = "Next batch is prepared"
    Phase4_UKMessage19<public>:string = "That looks great!"
    Phase4_UKMessage20<public>:string = "Send the Bread with the drone"
    Phase4_UKMessage21<public>:string = "Tutorial done! Back to Bakery"
    
    Extra_Message0<public>:string = "Pick already upgraded"
    Extra_Message1<public>:string = "Baker already upgraded"

tutorial_fr_module<public> := module:
    Phase0_UKMessage0<public>:string = "Bienvenue à Breaking Bread !"
    Phase0_UKMessage1<public>:string = "Le pain rassis donne des ingrédients clés"
    
    Phase1_UKMessage0<public>:string = "Ajoutez des ingrédients dans l'entonnoir"
    Phase1_UKMessage1<public>:string = "Utilisez l'entonnoir"
    Phase1_UKMessage2<public>:string = "Regardez vos dépôts se remplir"
    
    Phase2_UKMessage0<public>:string = "Découvrons les Boulangers"
    Phase2_UKMessage1<public>:string = "Rassemblez 5 Eaux"
    Phase2_UKMessage2<public>:string = "Rassemblez 5 Farines"
    Phase2_UKMessage3<public>:string = "Ajoutez-les dans l'entonnoir"
    
    Phase2_UKMessage4<public>:string = "Ouvrez les vannes pour envoyer les ingrédients"
    Phase2_UKMessage44<public>:string = "Vannes vertes > vannes jaunes"
    Phase2_UKMessage5<public>:string = "Gardez les barres pleines"
    Phase2_UKMessage6<public>:string = "Plus d'ingrédients ! Rassemblez-en"
    
    Phase2_UKMessage7<public>:string = "Bravo ! Le Boulanger travaille"
    Phase2_UKMessage8<public>:string = "Lumières vertes = cuisson en cours"
    Phase2_UKMessage9<public>:string = "Accélération temporaire"
    Phase2_UKMessage10<public>:string = "Wow ! C'est superbe !"
    Phase2_UKMessage11<public>:string = "Envoyez le pain avec le drone"
    
    Phase3_UKMessage0<public>:string = "Améliorez le Pic"
    Phase3_UKMessage1<public>:string = "Vous récoltez plus maintenant !"
    Phase3_UKMessage2<public>:string = "Testez-le sur du pain rassis"
    Phase3_UKMessage3<public>:string = "Génial !"
    Phase3_UKMessage4<public>:string = "Les barres montrent les besoins"
    Phase3_UKMessage5<public>:string = "Vérifiez les ressources du four"
    Phase3_UKMessage6<public>:string = "Interagissez avec l'ordinateur !"
    Phase3_UKMessage7<public>:string = "Avez-vous vu les quantités ?"
    Phase3_UKMessage8<public>:string = "Améliorez la production"
    Phase3_UKMessage9<public>:string = "L'amélioration augmente les besoins"
    
    Phase4_UKMessage0<public>:string = "Utilisez du charbon pour booster"
    Phase4_UKMessage1<public>:string = "Ajoutez 30 charbons dans l'entonnoir"
    Phase4_UKMessage2<public>:string = "Génère de l'énergie temporaire"
    Phase4_UKMessage3<public>:string = "Vérifiez l'énergie restante"
    Phase4_UKMessage4<public>:string = "Voyez votre multiplicateur ici"
    Phase4_UKMessage5<public>:string = "Le charbon donne de l'énergie partagée"
    Phase4_UKMessage6<public>:string = "Boost léger pour tous les joueurs"
    Phase4_UKMessage7<public>:string = "Tous boostés = bonus supplémentaire !"
    Phase4_UKMessage8<public>:string = "Appliquez ce que vous avez appris !"
    
    Phase4_UKMessage10<public>:string = "Ajoutez des ingrédients"
    Phase4_UKMessage11<public>:string = "Ajoutez du charbon avant d'ouvrir"
    Phase4_UKMessage12<public>:string = "Ouvrez les vannes maintenant"
    Phase4_UKMessage13<public>:string = "Ajoutez les ingrédients manquants et cuisez le pain !"
    Phase4_UKMessage14<public>:string = "Bravo ! Le pain cuit"
    Phase4_UKMessage15<public>:string = "Lumières vertes = cuisson active"
    Phase4_UKMessage16<public>:string = "Les besoins ont changé, regardez !"
    Phase4_UKMessage17<public>:string = "La prochaine fournée est prête"
    Phase4_UKMessage19<public>:string = "Ça a l'air délicieux !"
    Phase4_UKMessage20<public>:string = "Envoyez le pain avec le drone"
    Phase4_UKMessage21<public>:string = "Tutoriel fini ! Retour au four"
    
    Extra_Message0<public>:string = "Pic déjà amélioré"
    Extra_Message1<public>:string = "Four déjà amélioré"
    

tutorial_es_module<public> := module:
    Phase0_UKMessage0<public>:string = "¡Bienvenido a Breaking Bread!"
    Phase0_UKMessage1<public>:string = "El Pan Duro da ingredientes clave"
    
    Phase1_UKMessage0<public>:string = "Añade ingredientes al embudo"
    Phase1_UKMessage1<public>:string = "Usa el embudo"
    Phase1_UKMessage2<public>:string = "Mira cómo se llenan los depósitos"
    
    Phase2_UKMessage0<public>:string = "Aprendamos sobre los Panaderos"
    Phase2_UKMessage1<public>:string = "Reúne 5 de Agua"
    Phase2_UKMessage2<public>:string = "Reúne 5 de Harina"
    Phase2_UKMessage3<public>:string = "Agrégalos en el embudo"
    
    Phase2_UKMessage4<public>:string = "Abre las válvulas para enviar ingredientes"
    Phase2_UKMessage44<public>:string = "Válvulas verdes > válvulas amarillas"
    Phase2_UKMessage5<public>:string = "Mantén las barras llenas"
    Phase2_UKMessage6<public>:string = "¿Faltan ingredientes? ¡Consigue más!"
    
    Phase2_UKMessage7<public>:string = "¡Bien hecho! El Panadero trabaja"
    Phase2_UKMessage8<public>:string = "Luces verdes = horneando"
    Phase2_UKMessage9<public>:string = "Acelerando el Panadero un rato"
    Phase2_UKMessage10<public>:string = "¡Wow! ¡Se ve genial!"
    Phase2_UKMessage11<public>:string = "Envía el pan con el dron"
    
    Phase3_UKMessage0<public>:string = "Mejora el Pico de Recolección"
    Phase3_UKMessage1<public>:string = "¡Reúnes más por golpe ahora!"
    Phase3_UKMessage2<public>:string = "Prueba el daño en el Pan Duro"
    Phase3_UKMessage3<public>:string = "¡Genial!"
    Phase3_UKMessage4<public>:string = "Las barras muestran recursos faltantes"
    Phase3_UKMessage5<public>:string = "Consulta los recursos en la PC"
    Phase3_UKMessage6<public>:string = "¡Interactúa con ella!"
    Phase3_UKMessage7<public>:string = "¿Viste las cantidades necesarias?"
    Phase3_UKMessage8<public>:string = "Mejora la producción"
    Phase3_UKMessage9<public>:string = "Mejorar aumenta los requisitos"
    
    Phase4_UKMessage0<public>:string = "Usa Carbón para potenciar la producción"
    Phase4_UKMessage1<public>:string = "Añade 30 Carbones en el embudo"
    Phase4_UKMessage2<public>:string = "Genera Súper Energía un tiempo"
    Phase4_UKMessage3<public>:string = "Mira cuánta energía queda"
    Phase4_UKMessage4<public>:string = "Aquí ves tu multiplicador"
    Phase4_UKMessage5<public>:string = "El Carbón da Energía Compartida"
    Phase4_UKMessage6<public>:string = "Pequeño boost para todos"
    Phase4_UKMessage7<public>:string = "Si todos la tienen, ¡extra boost!"
    Phase4_UKMessage8<public>:string = "Hora de aplicar lo aprendido"
    
    Phase4_UKMessage10<public>:string = "Añade ingredientes al embudo"
    Phase4_UKMessage11<public>:string = "Añade Carbón antes de abrir válvulas"
    Phase4_UKMessage12<public>:string = "Abre las válvulas ahora"
    Phase4_UKMessage13<public>:string = "Agrega los ingredientes faltantes y hornea el pan!"
    Phase4_UKMessage14<public>:string = "¡Bien hecho! El panadero hornea"
    Phase4_UKMessage15<public>:string = "Luces verdes = funcionando"
    Phase4_UKMessage16<public>:string = "¿Notaste los cambios de necesidades?"
    Phase4_UKMessage17<public>:string = "La siguiente tanda se prepara"
    Phase4_UKMessage19<public>:string = "¡Se ve delicioso!"
    Phase4_UKMessage20<public>:string = "Envía el pan con el dron"
    Phase4_UKMessage21<public>:string = "¡Tutorial completado!"
    
    Extra_Message0<public>:string = "Pico ya mejorado"
    Extra_Message1<public>:string = "Panadero ya mejorado"

tutorial_pt_module<public> := module:
    Phase0_UKMessage0<public>:string = "Bem-vindo ao Breaking Bread!"
    Phase0_UKMessage1<public>:string = "O Pão Velho dá ingredientes chave"
    
    Phase1_UKMessage0<public>:string = "Coloque ingredientes no funil"
    Phase1_UKMessage1<public>:string = "Use o funil"
    Phase1_UKMessage2<public>:string = "Veja os depósitos enchendo"
    
    Phase2_UKMessage0<public>:string = "Vamos conhecer os Padeiros"
    Phase2_UKMessage1<public>:string = "Recolha 5 de Água"
    Phase2_UKMessage2<public>:string = "Recolha 5 de Farinha"
    Phase2_UKMessage3<public>:string = "Coloque no funil"
    
    Phase2_UKMessage4<public>:string = "Abra válvulas para enviar ingredientes"
    Phase2_UKMessage44<public>:string = "Válvulas verdes > válvulas amarelas"
    Phase2_UKMessage5<public>:string = "Mantenha as barras cheias"
    Phase2_UKMessage6<public>:string = "Faltam ingredientes? Pegue mais!"
    
    Phase2_UKMessage7<public>:string = "Ótimo! O Padeiro está trabalhando"
    Phase2_UKMessage8<public>:string = "Luzes verdes = assando"
    Phase2_UKMessage9<public>:string = "Acelerando o Padeiro por um tempo"
    Phase2_UKMessage10<public>:string = "Uau! Parece ótimo!"
    Phase2_UKMessage11<public>:string = "Envie o pão com o drone"
    
    Phase3_UKMessage0<public>:string = "Melhore a Picareta"
    Phase3_UKMessage1<public>:string = "Agora você coleta mais por golpe!"
    Phase3_UKMessage2<public>:string = "Teste no Pão Velho"
    Phase3_UKMessage3<public>:string = "Legal!"
    Phase3_UKMessage4<public>:string = "As barras mostram os recursos"
    Phase3_UKMessage5<public>:string = "Verifique os recursos do Padeiro"
    Phase3_UKMessage6<public>:string = "Interaja com o PC!"
    Phase3_UKMessage7<public>:string = "Viu as quantidades necessárias?"
    Phase3_UKMessage8<public>:string = "Melhore a produção"
    Phase3_UKMessage9<public>:string = "Melhorar aumenta a necessidade"
    
    Phase4_UKMessage0<public>:string = "Use carvão para melhorar a produção"
    Phase4_UKMessage1<public>:string = "Coloque 30 carvões no funil"
    Phase4_UKMessage2<public>:string = "Gera Super Energia por um tempo"
    Phase4_UKMessage3<public>:string = "Veja a energia restante"
    Phase4_UKMessage4<public>:string = "Veja seu multiplicador aqui"
    Phase4_UKMessage5<public>:string = "Carvão gera Energia Compartilhada"
    Phase4_UKMessage6<public>:string = "Pequeno boost para todos"
    Phase4_UKMessage7<public>:string = "Se todos tiverem, extra boost!"
    Phase4_UKMessage8<public>:string = "Hora de aplicar o que aprendeu!"
    
    Phase4_UKMessage10<public>:string = "Coloque ingredientes no funil"
    Phase4_UKMessage11<public>:string = "Adicione carvão antes das válvulas"
    Phase4_UKMessage12<public>:string = "Abra as válvulas agora"
    Phase4_UKMessage13<public>:string = "Adicione os ingredientes faltantes e asse o pão!"
    Phase4_UKMessage14<public>:string = "Ótimo! O pão está assando"
    Phase4_UKMessage15<public>:string = "Luzes verdes = funcionando"
    Phase4_UKMessage16<public>:string = "Notou mudanças nas necessidades?"
    Phase4_UKMessage17<public>:string = "O próximo lote está pronto"
    Phase4_UKMessage19<public>:string = "Parece delicioso!"
    Phase4_UKMessage20<public>:string = "Envie o pão com o drone"
    Phase4_UKMessage21<public>:string = "Tutorial concluído!"
    
    Extra_Message0<public>:string = "Picareta já melhorada"
    Extra_Message1<public>:string = "Padeiro já melhorado"



otherstrings_en_module<public> := module:
    GoodLuck<public>:string = "Good luck!"
    Recomendation<public>:string = "Don't skip the tutorial!"
    Title<public>:string = "Tutorial"
    Funnel<public>:string = "Use the funnel"
    GatherWater<public>:string = "Gather 5 Water"
    GatherFlour<public>:string = "Gather 5 Flour"
    CheckPick<public>:string = "Check the Pick"
    GoodJob<public>:string = "Good job!"
    OpenValves<public>:string = "Open the valves"
    UseCoal<public>:string = "Add 30 Coal"
    SendDrone<public>:string = "Send the drone"
    CanNotDrone<public>:string = "Can't send the drone now"
    UpgradePick<public>:string = "Upgrade the Pick"
    CheckComputer<public>:string = "Check the computer"
    BakeBread<public>:string = "Bake Bread"
    ImproveProduction<public>:string = "Boost Baker's production"
    Problem<public>:string = "Let's fix it and continue!"
    ThisTheBakery<public>:string = "This is your Bakery"
    CompleteAgain<public>:string = "Redo tutorial if needed"
    Shop<public>:string = "Buy upgrades in the shop"
    Pets<public>:string = "Complete Quests to unlock Sourdoughs (Pets)"

otherstrings_fr_module<public> := module:
    GoodLuck<public>:string = "Bonne chance !"
    Recomendation<public>:string = "Ne sautez pas le tutoriel !"
    Title<public>:string = "Tutoriel"
    Funnel<public>:string = "Utilisez l'entonnoir"
    GatherWater<public>:string = "Récupérez 5 eaux"
    GatherFlour<public>:string = "Récupérez 5 farines"
    CheckPick<public>:string = "Vérifiez la pioche"
    GoodJob<public>:string = "Bon travail !"
    OpenValves<public>:string = "Ouvrez les vannes"
    UseCoal<public>:string = "Ajoutez 30 charbons"
    SendDrone<public>:string = "Envoyez le drone"
    CanNotDrone<public>:string = "Impossible d'envoyer le drone"
    UpgradePick<public>:string = "Améliorez la pioche"
    CheckComputer<public>:string = "Vérifiez l'ordinateur"
    BakeBread<public>:string = "Faites du pain"
    ImproveProduction<public>:string = "Boostez la production du boulanger"
    Problem<public>:string = "Corrigeons cela et continuons !"
    ThisTheBakery<public>:string = "Voici votre boulangerie"
    CompleteAgain<public>:string = "Refaire le tutoriel si besoin"
    Shop<public>:string = "Achetez des améliorations"
    Pets<public>:string = "Terminez des quêtes pour débloquer des animaux"

otherstrings_es_module<public> := module:
    GoodLuck<public>:string = "¡Buena suerte!"
    Recomendation<public>:string = "¡No te saltes el tutorial!"
    Title<public>:string = "Tutorial"
    Funnel<public>:string = "Usa el embudo"
    GatherWater<public>:string = "Recoge 5 de agua"
    GatherFlour<public>:string = "Recoge 5 de harina"
    CheckPick<public>:string = "Revisa el pico"
    GoodJob<public>:string = "¡Buen trabajo!"
    OpenValves<public>:string = "Abre las válvulas"
    UseCoal<public>:string = "Añade 30 carbones"
    SendDrone<public>:string = "Envía el dron"
    CanNotDrone<public>:string = "No puedes enviar el dron ahora"
    UpgradePick<public>:string = "Mejora el pico"
    CheckComputer<public>:string = "Revisa la computadora"
    BakeBread<public>:string = "Hornea pan"
    ImproveProduction<public>:string = "Aumenta la producción del panadero"
    Problem<public>:string = "¡Arreglémoslo y sigamos!"
    ThisTheBakery<public>:string = "Esta es tu panadería"
    CompleteAgain<public>:string = "Repite el tutorial si lo necesitas"
    Shop<public>:string = "Compra mejoras en la tienda"
    Pets<public>:string = "Completa misiones para desbloquear mascotas"

otherstrings_pt_module<public> := module:
    GoodLuck<public>:string = "Boa sorte!"
    Recomendation<public>:string = "Não pule o tutorial!"
    Title<public>:string = "Tutorial"
    Funnel<public>:string = "Use o funil"
    GatherWater<public>:string = "Colete 5 de água"
    GatherFlour<public>:string = "Colete 5 de farinha"
    CheckPick<public>:string = "Verifique a picareta"
    GoodJob<public>:string = "Bom trabalho!"
    OpenValves<public>:string = "Abra as válvulas"
    UseCoal<public>:string = "Adicione 30 carvões"
    SendDrone<public>:string = "Envie o drone"
    CanNotDrone<public>:string = "Não pode enviar o drone agora"
    UpgradePick<public>:string = "Melhore a picareta"
    CheckComputer<public>:string = "Verifique o computador"
    BakeBread<public>:string = "Asse pão"
    ImproveProduction<public>:string = "Aumente a produção do padeiro"
    Problem<public>:string = "Vamos corrigir e continuar!"
    ThisTheBakery<public>:string = "Esta é sua padaria"
    CompleteAgain<public>:string = "Refaça o tutorial se necessário"
    Shop<public>:string = "Compre melhorias na loja"
    Pets<public>:string = "Complete missões para desbloquear mascotes"



# A Verse-authored creative device that can be placed in a level
uk_tutorial := class(creative_device):

    Mat_RedScreen : material = Materials.RedScreen_Inst
    Mat_GreenScreen : material = Materials.GreenScreen_Inst

    @editable
    SuperEnergy_Timer : timer_device = timer_device{}
    @editable
    OverPoweredSign : creative_prop = creative_prop{}
    var SuperEnergy : float = 0.0
    var SuperEnergy_Multiplier : float = 1.0

    #Drone state
    var IsDroneReady : logic = true
    var IsPickUpgraded : logic = false
    var IsBakerUpgraded : logic = false

    #Valve State
    var ValveState : int = 0
    var WaterPreparedFor_B : float = 0.0
    var FlourPreparedFor_B : float = 0.0

    var WaterReqFor_B : float = 10.0
    var FlourReqFor_B : float = 10.0

    #UkResourcers
    var MainDeposit_Water : float = 0.0
    var MainDeposit_Flour : float = 0.0
    var MainDeposit_Cheese : float = 0.0
    var MainDeposit_Pimentos : float = 0.0

    #PlayerResourcers
    var PlayerWaterInventory : float = 0.0
    var PlayerFlourInventory : float = 0.0
    var PlayerCheeseInventory : float = 0.0
    var PlayerPimentosInventory : float = 0.0

    #CurrentTutorialObjetive
    var CurrentPhase : int = -1
    var CurrentTutorialObjetive_Type : int = -1
    var CurrentTutorialObjetive_Amount : float = -1.0
    var CurrentTutorialProgress_Amount : float = -1.0
    var CurrentTutorial_Icon : texture = Images.Water_Icon
    #Phase0
    var Phase0State : int = -1
    #Phase1
    var Phase1State : int = 0
    #Phase2
    var Phase2State : int = 0
    #Phase3
    var Phase3State : int = 0
    #Phase4
    var Phase4State : int = 0

    var MaybeBakery : ?bakery_machines_manager = false
    var ConfirmationToExit : logic = false

    #CommonButtons
    @editable
    CoalButton : button_device = button_device{}
    @editable
    BakerButton : button_device = button_device{}
    @editable
    BakersInfoButton : button_device = button_device{}
    @editable
    Pick : button_device = button_device{}
    @editable
    PhasePosition :  creative_prop = creative_prop{}
    @editable
    Valve : button_device = button_device{}
    @editable
    WaterValve_Prop : creative_prop = creative_prop{}
    @editable
    FlourValve_Prop : creative_prop = creative_prop{}
    @editable
    Baker : creative_prop = creative_prop{}
    @editable
    Funnel : button_device = button_device{}
    @editable
    DroneProp : creative_prop = creative_prop{}
    @editable
    DroneButton : button_device = button_device{}
    var AllDrones_InitialTransform : []transform = array{}
    @editable
    ExitButton : button_device = button_device{}

    #Deposits
    <#
    AllWaterDeposit_Billboard : []billboard_device = array{}
    AllFlourDeposit_Billboard : []billboard_device = array{}
    AllCheeseDeposit_Billboard : []billboard_device = array{}
    AllPimentosDeposit_Billboard : []billboard_device = array{}#>

    #editable
    @editable
    AllWaterDeposit_Props : []creative_prop = array{}
    @editable
    AllFlourDeposit_Props : []creative_prop = array{}
    @editable
    AllCheeseDeposit_Props : []creative_prop = array{}
    @editable
    AllPimentosDeposit_Props : []creative_prop = array{}

    #Phase 0
    #Teleporters
    #AllCameras
    @editable
    AllCameraPhase0 : []gameplay_camera_fixed_point_device = array{}

    #Phase 1
    #Interactables
    #AllCameras
    @editable
    AllCameraPhase1 : []gameplay_camera_fixed_point_device = array{}

    #Phase 2
    #AllCameras
    @editable
    AllCameraPhase2 : []gameplay_camera_fixed_point_device = array{}

    #Phase 3
    #AllCameras
    @editable
    AllCameraPhase3 : []gameplay_camera_fixed_point_device = array{}

    #Phase 4
    #AllCameras
    @editable
    AllCameraPhase4 : []gameplay_camera_fixed_point_device = array{}

    # Runs when the device is started in a running game
    OnBegin<override>()<suspends>:void=
        InitSubscribes()
        set AllDrones_InitialTransform = AllDrones_InitialTransform + array{DroneProp.GetTransform()}


    UpdateSubdeposits() : void =
        if:
            Percentage_WaterPer_B := Floor[(WaterPreparedFor_B/WaterReqFor_B)*4]
            Percentage_FlourPer_B := Floor[(FlourPreparedFor_B/FlourReqFor_B)*4]
        then:
            for(Index:=2..5):
                #Bread
                if(Percentage_WaterPer_B+1  >= Index):
                    WaterValve_Prop.SetMaterial(Mat_GreenScreen, ?Index:=Index)
                else:
                    WaterValve_Prop.SetMaterial(Mat_RedScreen, ?Index:=Index)

                if(Percentage_FlourPer_B+1  >= Index):
                    FlourValve_Prop.SetMaterial(Mat_GreenScreen, ?Index:=Index)
                else:
                    FlourValve_Prop.SetMaterial(Mat_RedScreen, ?Index:=Index)
        else:
            PrintFatalError("On UpdateSubdeposits(): percentage Failed")
        UpdateAllMainDeposits()

    BakeCompletedEvent() : void = 
        if(ValidBakery:=MaybeBakery?):
            if(COMPLETEBAKE_TUTORIAL_TYPE=CurrentTutorialObjetive_Type):
                set CurrentTutorialObjetive_Type = -1
                AddPhaseProgress()

    InfoCheckedEvent() : void =
        if(ValidBakery:=MaybeBakery?):
            if(CHECKINFO_TUTORIAL_TYPE=CurrentTutorialObjetive_Type):
                AddPhaseProgress()
    LanguageSelectedEvent() : void =
        if(ValidBakery:=MaybeBakery?):
            if(LANGUAGE_TUTORIAL_TYPE=CurrentTutorialObjetive_Type):
                AddPhaseProgress()

    TryConsumeSuperEnergy() : void =
        #OwnedEnergy
        set SuperEnergy = SuperEnergy - 1.0 #Consume
        #Set Timer  
        if(SuperEnergy <= 0.0):
            SuperEnergy_Timer.SetActiveDuration(0.0)
            set SuperEnergy = 0.0
            set SuperEnergy_Multiplier = 1.0
        else:  
            SuperEnergy_Timer.SetMaxDuration(SuperEnergy) 
            set SuperEnergy_Multiplier = 2.0
        SetSuperEnergyMultiplierProp()

    SetSuperEnergyMultiplierProp<private>() : void =
        #Log("SuperEnergy_Multiplier: {SuperEnergy_Multiplier}", LOG_LEVEL_COAL)
        if:
            Unidades := Floor[SuperEnergy_Multiplier/1.0]
            Decimal := Floor[(SuperEnergy_Multiplier-Unidades*1.0)/0.1]
        then:
            Remaining := SuperEnergy_Multiplier - (Unidades*1.0) - (Decimal*0.1)
            #Log("Remaining: {Remaining}", LOG_LEVEL_COAL)
            #Centesimal := Floor[(SuperEnergy_Multiplier-(Unidades*1.0)-(Decimal*0.1))/0.01]
            CentesimalFloat := Remaining/0.01
            #Log("CentesimalFloat: {CentesimalFloat}", LOG_LEVEL_COAL)

            if(CentesimalInt := Ceil[CentesimalFloat]):
                #Log("Unidades: {Unidades}", LOG_LEVEL_COAL)
                #Log("Decimal: {Decimal}", LOG_LEVEL_COAL)
                #Log("Centesimal: {CentesimalInt}", LOG_LEVEL_COAL)

                if:
                    UnidadesMaterial:=AllMaterialnumbers[Unidades]
                    DecimalMaterial:=AllMaterialnumbers[Decimal]
                    CentesimalMaterial:=AllMaterialnumbers[CentesimalInt]
                then:
                    OverPoweredSign.SetMaterial(UnidadesMaterial,?Index:=2)
                    OverPoweredSign.SetMaterial(DecimalMaterial,?Index:=3)      
                    OverPoweredSign.SetMaterial(CentesimalMaterial,?Index:=4)  
        else:
            PrintFatalError("Fatal error on SetSuperEnergyMultiplierProp")

    InsertedCoalEvent(Amount : float) : void =
        if(ValidBakery:=MaybeBakery?):
            Log("Inserted Coal EVENT!",LOG_LEVEL_UK)
            if(USECOAL_TUTORIAL_TYPE=CurrentTutorialObjetive_Type):
                Log("The current event is to insert coal, adding progress",LOG_LEVEL_UK)
                set CurrentTutorialProgress_Amount = CurrentTutorialProgress_Amount + Amount
                ValidBakery.SetUKQuest()
                if(CurrentTutorialProgress_Amount>=CurrentTutorialObjetive_Amount):
                    if(DropCoal:=ValidBakery.MaybeDropCoalUI?):
                        DropCoal.Remove_CanvasWithoutMessage()  
                    AddPhaseProgress()
                
        if(SuperEnergy>0.0):
            set SuperEnergy = SuperEnergy + Amount
        else:
            set SuperEnergy = Amount
            spawn:
                SpawnSuperEnergy()

    SpawnSuperEnergy()<suspends> : void =
        loop:
            Sleep(1.0)
            TryConsumeSuperEnergy()
            if(SuperEnergy<=0.0):
                break



    PickUpgradedEvent() : void = 
        if(ValidBakery:=MaybeBakery?):
            if(IMPROVEPICKAXE_TUTORIAL_TYPE=CurrentTutorialObjetive_Type):
                set CurrentTutorialObjetive_Type = -1
                set IsPickUpgraded = true
                AddPhaseProgress()

    BakerUpgradedEvent() : void = 
        if(ValidBakery:=MaybeBakery?):
            if(IMPROVEBAKERY_TUTORIAL_TYPE=CurrentTutorialObjetive_Type):
                set CurrentTutorialObjetive_Type = -1
                set IsBakerUpgraded = true
                set WaterReqFor_B = 15.0
                set FlourReqFor_B = 15.0
                AddPhaseProgress()
                UpdateBakery()

    PickHitEvent() : void =
        if(ValidBakery:=MaybeBakery?):
            if(PICKHIT_TUTORIAL_TYPE=CurrentTutorialObjetive_Type):
                AddPhaseProgress()

    ValveInteractedEvent(Agent : agent) : void = 
        if(ValidBakery:=MaybeBakery?):
            set ValveState = ValveState + 1
            if(ValveState>2):
                set ValveState = 0
            UpdateValveProps()
            ValidBakery.GameManager.VFXManager.Audio_ValveChanged.Play(Agent)
            TakeResourcersFromDeposits()
            if(USEVALVES_TUTORIAL_TYPE=CurrentTutorialObjetive_Type):
                set CurrentTutorialObjetive_Type = -1
                AddPhaseProgress()
            

    UpdateValveProps() : void =
            if(ValveState = 0):
                WaterValve_Prop.SetMaterial(Materials.Mat_PipeState0, ?Index:=0)
            if(ValveState = 1):
                WaterValve_Prop.SetMaterial(Materials.Mat_PipeState1, ?Index:=0)
            if(ValveState = 2):
                WaterValve_Prop.SetMaterial(Materials.Mat_PipeState2, ?Index:=0)

            if(ValveState = 0):
                FlourValve_Prop.SetMaterial(Materials.Mat_PipeState0, ?Index:=0)
            if(ValveState = 1):
                FlourValve_Prop.SetMaterial(Materials.Mat_PipeState1, ?Index:=0)
            if(ValveState = 2):
                FlourValve_Prop.SetMaterial(Materials.Mat_PipeState2, ?Index:=0)


    UpdateAllMainDeposits() : void =
        <#
        for(WaterDeposit_Billbord : AllWaterDeposit_Billboard):
            WaterDeposit_Billbord.SetText(BasicMessage("{GetOptimizedMagnitudeLetter(MainDeposit_Water)}\n―――\n{GetOptimizedMagnitudeLetter(10.0)}"))
        for(FlourDeposit_Billbord : AllFlourDeposit_Billboard):
            FlourDeposit_Billbord.SetText(BasicMessage("{GetOptimizedMagnitudeLetter(MainDeposit_Flour)}\n―――\n{GetOptimizedMagnitudeLetter(10.0)}"))
        for(CheeseDeposit_Billbord : AllCheeseDeposit_Billboard):
            CheeseDeposit_Billbord.SetText(BasicMessage("{GetOptimizedMagnitudeLetter(MainDeposit_Cheese)}\n―――\n{GetOptimizedMagnitudeLetter(10.0)}"))
        for(PimentosDeposit_Billbord : AllPimentosDeposit_Billboard):
            PimentosDeposit_Billbord.SetText(BasicMessage("{GetOptimizedMagnitudeLetter(MainDeposit_Pimentos)}\n―――\n{GetOptimizedMagnitudeLetter(10.0)}"))#>

        Log("MainDeposit_Water: {MainDeposit_Water}",LOG_LEVEL_UK)
        Log("MainDeposit_Flour: {MainDeposit_Flour}",LOG_LEVEL_UK)

        for(MainWaterDeposit_AmmountProp : AllWaterDeposit_Props):
            spawn:
                AnimateDeposit(MainWaterDeposit_AmmountProp,10.0,MainDeposit_Water, false)
        for(MainFlourDeposit_AmmountProp : AllFlourDeposit_Props):
            spawn:
                AnimateDeposit(MainFlourDeposit_AmmountProp,10.0,MainDeposit_Flour, false)
        for(MainCheeseDeposit_AmmountProp : AllCheeseDeposit_Props):
            spawn:
                AnimateDeposit(MainCheeseDeposit_AmmountProp,10.0,MainDeposit_Cheese, false)
        for(MainPimentosDeposit_AmmountProp : AllPimentosDeposit_Props):
            spawn:
                AnimateDeposit(MainPimentosDeposit_AmmountProp,10.0,MainDeposit_Pimentos, false)

            #var NewData : player_data_table = PlayersDataManager.SetDepositResourcers(PlayerData,(MainWaterDeposit_Value,MainFlourDeposit_Value,MainCheeseDeposit_Value,MainPimentosDeposit_Value))
            #set NewData = PlayersDataManager.SetSubDepositResourcers(NewData,(WaterPreparedFor_B,FlourPreparedFor_B,WaterPreparedFor_CB,FlourPreparedFor_CB,CheesePreparedFor_CB,WaterPreparedFor_PCS,FlourPreparedFor_PCS,CheesePreparedFor_PCS,PimentosPreparedFor_PCS))
            #PlayersDataManager.SaveNewData(Player,PlayerData,NewData)

    SetBakery(BakeryMachines : bakery_machines_manager) : void =
        set MaybeBakery = option{BakeryMachines}

    InitSubscribes() : void =
        CoalButton.InteractedWithEvent.Subscribe(CoalInteractedEvent)
        BakersInfoButton.InteractedWithEvent.Subscribe(DisplayBakerInfoButton)
        BakerButton.InteractedWithEvent.Subscribe(DisplayBakerUpgrades)
        Pick.InteractedWithEvent.Subscribe(DisplayPickUpgrades)
        DroneButton.InteractedWithEvent.Subscribe(DroneInteractedEvent)
        Funnel.InteractedWithEvent.Subscribe(FunnelInteractedEvent)
        Valve.InteractedWithEvent.Subscribe(ValveInteractedEvent)
        
        ExitButton.InteractedWithEvent.Subscribe(TryExitTutorial)

    DisplayBakerInfoButton(Agent : agent) : void =
        if(ValidBakery:=MaybeBakery?):
            ValidBakery.Display_InfoBB(Agent)

    DisplayPickUpgrades(Agent : agent) : void =
        if(ValidBakery:=MaybeBakery?):
            if(not IsPickUpgraded?):
                ValidBakery.Display_PickaxeUpgrades(Agent)
            else:
                if(ValidBakery.PlayerDataCache.Language=ENGLISH):
                    ValidBakery.GameManager.DisplayCharacterMessageLoc(Agent,I_CHARACTER,BasicMessage(tutorial_en_module.Extra_Message0),?Sound:=ERROR_SOUND,?DisplayTime:=3.0,?Priority:=MIN_PRIORITY)
                else if(ValidBakery.PlayerDataCache.Language=SPANISH):
                    ValidBakery.GameManager.DisplayCharacterMessageLoc(Agent,I_CHARACTER,BasicMessage(tutorial_es_module.Extra_Message0),?Sound:=ERROR_SOUND,?DisplayTime:=3.0,?Priority:=MIN_PRIORITY)
                else if(ValidBakery.PlayerDataCache.Language=FRENCH):
                    ValidBakery.GameManager.DisplayCharacterMessageLoc(Agent,I_CHARACTER,BasicMessage(tutorial_fr_module.Extra_Message0),?Sound:=ERROR_SOUND,?DisplayTime:=3.0,?Priority:=MIN_PRIORITY)
                else if(ValidBakery.PlayerDataCache.Language=PORTUGUESE):
                    ValidBakery.GameManager.DisplayCharacterMessageLoc(Agent,I_CHARACTER,BasicMessage(tutorial_pt_module.Extra_Message0),?Sound:=ERROR_SOUND,?DisplayTime:=3.0,?Priority:=MIN_PRIORITY)


    DisplayBakerUpgrades(Agent : agent) : void =
        if(ValidBakery:=MaybeBakery?):
            if(not IsBakerUpgraded?):
                ValidBakery.Display_ComputerBBInfo(Agent)
            else:
                if(ValidBakery.PlayerDataCache.Language=ENGLISH):
                    ValidBakery.GameManager.DisplayCharacterMessageLoc(Agent,I_CHARACTER,BasicMessage(tutorial_en_module.Extra_Message1),?Sound:=ERROR_SOUND,?DisplayTime:=3.0,?Priority:=MIN_PRIORITY)
                else if(ValidBakery.PlayerDataCache.Language=SPANISH):
                    ValidBakery.GameManager.DisplayCharacterMessageLoc(Agent,I_CHARACTER,BasicMessage(tutorial_es_module.Extra_Message1),?Sound:=ERROR_SOUND,?DisplayTime:=3.0,?Priority:=MIN_PRIORITY)
                else if(ValidBakery.PlayerDataCache.Language=FRENCH):
                    ValidBakery.GameManager.DisplayCharacterMessageLoc(Agent,I_CHARACTER,BasicMessage(tutorial_fr_module.Extra_Message1),?Sound:=ERROR_SOUND,?DisplayTime:=3.0,?Priority:=MIN_PRIORITY)
                else if(ValidBakery.PlayerDataCache.Language=PORTUGUESE):
                    ValidBakery.GameManager.DisplayCharacterMessageLoc(Agent,I_CHARACTER,BasicMessage(tutorial_pt_module.Extra_Message1),?Sound:=ERROR_SOUND,?DisplayTime:=3.0,?Priority:=MIN_PRIORITY)
                    

    FunnelInteractedEvent(Agent : agent) : void =
        if(ValidBakery:=MaybeBakery?):
            ValidBakery.DisplayResourcersGather(array{-PlayerWaterInventory,-PlayerFlourInventory,-PlayerCheeseInventory,-PlayerPimentosInventory},array{WATER,FLOUR,CHEESE,PIMENTOS},false)
            
            set MainDeposit_Water = MainDeposit_Water + PlayerWaterInventory
            set MainDeposit_Flour = MainDeposit_Flour + PlayerFlourInventory
            set MainDeposit_Cheese = MainDeposit_Cheese + PlayerCheeseInventory
            set MainDeposit_Pimentos = MainDeposit_Pimentos + PlayerPimentosInventory

            if(MainDeposit_Water>10.0):
                set MainDeposit_Water = 10.0
            if(MainDeposit_Flour>10.0):
                set MainDeposit_Flour = 10.0
            if(MainDeposit_Cheese>10.0):
                set MainDeposit_Cheese = 10.0
            if(MainDeposit_Pimentos>10.0):
                set MainDeposit_Pimentos = 10.0
            
            set PlayerWaterInventory = 0.0
            set PlayerFlourInventory = 0.0
            set PlayerCheeseInventory = 0.0
            set PlayerPimentosInventory = 0.0

            TakeResourcersFromDeposits()
            ValidBakery.RefreshUKResourcers(array{WATER,FLOUR},array{PlayerWaterInventory,PlayerFlourInventory})

            if(USEFUNNEL_TUTORIAL_TYPE=CurrentTutorialObjetive_Type):
                set CurrentTutorialObjetive_Type = -1
                AddPhaseProgress()

    UpdateBakery() : void =
        UpdateSubdeposits() 
        UpdateAllMainDeposits()
        UpdateValveProps()

    TakeResourcersFromDeposits() : void =
        if(ValveState > 0):
            if(WaterPreparedFor_B<WaterReqFor_B, MainDeposit_Water > 0.0):
                set WaterPreparedFor_B = WaterPreparedFor_B + MainDeposit_Water
                if(WaterPreparedFor_B>WaterReqFor_B):
                    Excedent := WaterPreparedFor_B - WaterReqFor_B
                    set MainDeposit_Water = Excedent
                else:
                    set MainDeposit_Water = 0.0
            if(FlourPreparedFor_B<FlourReqFor_B, MainDeposit_Flour > 0.0):
                set FlourPreparedFor_B = FlourPreparedFor_B + MainDeposit_Flour
                if(FlourPreparedFor_B>FlourReqFor_B):
                    Excedent := FlourPreparedFor_B - FlourReqFor_B
                    set MainDeposit_Flour = Excedent
                else:
                    set MainDeposit_Flour = 0.0
        UpdateBakery()

    CoalInteractedEvent(Agent : agent) : void =
        if(ValidBakery:=MaybeBakery?):
            ValidBakery.Display_DropCoal(Agent)

    DroneInteractedEvent(Agent : agent) : void =
        if(SENDDRONE_TUTORIAL_TYPE=CurrentTutorialObjetive_Type):
            set CurrentTutorialObjetive_Type = -1
            AddPhaseProgress()
            if(ValidBakery:=MaybeBakery?, IsDroneReady?):
                if(DroneInitialTransform := AllDrones_InitialTransform[0]):
                    spawn:
                        ValidBakery.GameManager.VFXManager.AnimateDroneMovement(0,DroneProp,DroneInitialTransform,false)
            spawn:
                PlayDroneReturn()
        else:
            if(ValidBakery:=MaybeBakery?):
                if(ValidBakery.PlayerDataCache.Language=ENGLISH):
                    ValidBakery.GameManager.DisplayCharacterMessageLoc(Agent,I_CHARACTER,BasicMessage(otherstrings_en_module.CanNotDrone),?Sound:=ERROR_SOUND,?DisplayTime:=3.0,?Priority:=MIN_PRIORITY)
                else if(ValidBakery.PlayerDataCache.Language=SPANISH):
                    ValidBakery.GameManager.DisplayCharacterMessageLoc(Agent,I_CHARACTER,BasicMessage(otherstrings_es_module.CanNotDrone),?Sound:=ERROR_SOUND,?DisplayTime:=3.0,?Priority:=MIN_PRIORITY)
                else if(ValidBakery.PlayerDataCache.Language=FRENCH):
                    ValidBakery.GameManager.DisplayCharacterMessageLoc(Agent,I_CHARACTER,BasicMessage(otherstrings_fr_module.CanNotDrone),?Sound:=ERROR_SOUND,?DisplayTime:=3.0,?Priority:=MIN_PRIORITY)
                else if(ValidBakery.PlayerDataCache.Language=PORTUGUESE):
                    ValidBakery.GameManager.DisplayCharacterMessageLoc(Agent,I_CHARACTER,BasicMessage(otherstrings_pt_module.CanNotDrone),?Sound:=ERROR_SOUND,?DisplayTime:=3.0,?Priority:=MIN_PRIORITY)
            
            
                

    PlayDroneReturn()<suspends> : void =
        set IsDroneReady = true
        Sleep(7.0)
        if(ValidBakery:=MaybeBakery?):
            if(DronesInitialTransform := AllDrones_InitialTransform[0]):
                spawn:
                    ValidBakery.GameManager.VFXManager.AnimateDroneMovement(1,DroneProp,DronesInitialTransform,false)

    AnimateBake() : void =
        if(ValidBakery:=MaybeBakery?):
            spawn:
                ValidBakery.GameManager.VFXManager.AnimateBake(Baker.GetTransform(),5.0,false,B_TYPE)

    ResourcersGatherEvent(AllResourcersType : []int, AllResourcersAmmount : []float) : void =
        if(ValidBakery:=MaybeBakery?):
            for(Index:=0..AllResourcersType.Length-1 , ResourcersType := AllResourcersType[Index], Amount := AllResourcersAmmount[Index]):
                if(ResourcersType=WATER):
                    set PlayerWaterInventory = PlayerWaterInventory + Amount
                    if(PlayerWaterInventory>5.0):
                        set PlayerWaterInventory = 5.0
                else if(ResourcersType=FLOUR):
                    set PlayerFlourInventory = PlayerFlourInventory + Amount
                    if(PlayerFlourInventory>5.0):
                        set PlayerFlourInventory = 5.0
                else if(ResourcersType=CHEESE):
                    set PlayerCheeseInventory = PlayerCheeseInventory + Amount
                    if(PlayerCheeseInventory>5.0):
                        set PlayerCheeseInventory = 5.0
                else if(ResourcersType=PIMENTOS):
                    set PlayerPimentosInventory = PlayerPimentosInventory + Amount
                    if(PlayerPimentosInventory>5.0):
                        set PlayerCheeseInventory = 5.0
                if(ResourcersType = CurrentTutorialObjetive_Type):
                    #set CurrentTutorialObjetive_Type = -1
                    set CurrentTutorialProgress_Amount = CurrentTutorialProgress_Amount + Amount
                    Log("UK Objetive added",LOG_LEVEL_UK)
                    if(CurrentTutorialObjetive_Amount<=CurrentTutorialProgress_Amount):
                        Log("UK Objetive completed",LOG_LEVEL_UK)
                        ValidBakery.SetUKQuest()
                        AddPhaseProgress()
                    else:
                        ValidBakery.SetUKQuest()
            ValidBakery.RefreshUKResourcers(array{WATER,FLOUR,CHEESE,PIMENTOS},array{PlayerWaterInventory,PlayerFlourInventory,PlayerCheeseInventory,PlayerPimentosInventory})

    PlayUKTutorial(Agent : agent) : void =
        Baker.SetMaterial(Materials.MI_Complement_red, ?Index:=1)
        if(ValidBakery:=MaybeBakery?):
            ValidBakery.SetUKTutorial(true)
        #Buttons
        CoalButton.Disable()
        BakerButton.Disable()
        BakersInfoButton.Disable()
        Pick.Disable()
        Valve.Disable()
        Funnel.Disable()
        DroneButton.Disable()

        #Requirements
        set WaterReqFor_B = 10.0
        set FlourReqFor_B = 10.0

        #Valve State
        set IsPickUpgraded = false
        set IsBakerUpgraded = false
        set ValveState = 0

        #Subdeposits
        set WaterPreparedFor_B = 0.0
        set FlourPreparedFor_B = 0.0

        #Inventory
        set PlayerWaterInventory = 0.0
        set PlayerFlourInventory = 0.0
        set PlayerCheeseInventory = 0.0
        set PlayerPimentosInventory = 0.0

        #Main deposit
        set MainDeposit_Water = 0.0
        set MainDeposit_Flour = 0.0
        set MainDeposit_Cheese = 0.0
        set MainDeposit_Pimentos = 0.0
        set CurrentPhase = 0
        set Phase0State = -1
        set Phase1State = 0
        set Phase2State = 0
        set Phase3State = 0
        set Phase4State = 0
        set ConfirmationToExit = false
        UpdateAllMainDeposits()
        PlayPhase0()

    TeleportToCurrentPhase(Agent : agent) : void =
        if(ValidBakery:=MaybeBakery?, Character := Agent.GetFortCharacter[]):
            ValidBakery.GameManager.ClassTeamSelector_Tutorial.ChangeClass(Agent)
            TransformAssigned := AdaptTransformFromVisualEmpty(PhasePosition.GetTransform(),VisualEmptyHideZ)
            ValidBakery.GameManager.TeleporterManager.TeleportPlayerUsingTeleporter(Agent,TransformAssigned)

    #Phase0
    PlayPhase0() : void =
        if(ValidBakery:=MaybeBakery?, Agent := ValidBakery.MaybeAgentOwner?):
            #teleport Player
            Log("Phase0State: {Phase0State}",LOG_LEVEL_UK)
            if(ValidBakery.PlayerDataCache.Language=ENGLISH):
            
            <#
            else if(ValidBakery.PlayerDataCache.Language=SPANISH):
                ValidBakery.SetUKQuestDescription(BasicMessage(otherstrings_es_module.GoodLuck))
            else if(ValidBakery.PlayerDataCache.Language=FRENCH):
                ValidBakery.SetUKQuestDescription(BasicMessage(otherstrings_fr_module.GoodLuck))
            else if(ValidBakery.PlayerDataCache.Language=PORTUGUESE):
                ValidBakery.SetUKQuestDescription(BasicMessage(otherstrings_pt_module.GoodLuck))#>

            if(Phase0State=-1):
                ValidBakery.ForceLanguageSelection()
                set CurrentTutorialObjetive_Type = LANGUAGE_TUTORIAL_TYPE
                set CurrentTutorialProgress_Amount = 0.0
                set CurrentTutorialObjetive_Amount = 1.0
    
            else if(Phase0State=0): #Hello
                ValidBakery.SetUKQuestDescription(othermessages_module.GoodLuck())
                ValidBakery.HideUKQuestDescription()
                TeleportToCurrentPhase(Agent)
                if(ValidBakery.PlayerDataCache.Language=ENGLISH):
                    ValidBakery.GameManager.DisplayCharacterMessageLoc(Agent,I_CHARACTER,BasicMessage(tutorial_en_module.Phase0_UKMessage0),?Sound:=MESSAGE_SOUND,?DisplayTime:=6.0,?Priority:=MAX_PRIORITY)
                else if(ValidBakery.PlayerDataCache.Language=SPANISH):
                    ValidBakery.GameManager.DisplayCharacterMessageLoc(Agent,I_CHARACTER,BasicMessage(tutorial_es_module.Phase0_UKMessage0),?Sound:=MESSAGE_SOUND,?DisplayTime:=6.0,?Priority:=MAX_PRIORITY)
                else if(ValidBakery.PlayerDataCache.Language=FRENCH):
                    ValidBakery.GameManager.DisplayCharacterMessageLoc(Agent,I_CHARACTER,BasicMessage(tutorial_fr_module.Phase0_UKMessage0),?Sound:=MESSAGE_SOUND,?DisplayTime:=6.0,?Priority:=MAX_PRIORITY)
                else if(ValidBakery.PlayerDataCache.Language=PORTUGUESE):
                    ValidBakery.GameManager.DisplayCharacterMessageLoc(Agent,I_CHARACTER,BasicMessage(tutorial_pt_module.Phase0_UKMessage0),?Sound:=MESSAGE_SOUND,?DisplayTime:=6.0,?Priority:=MAX_PRIORITY)
                
                spawn:
                    AddDelayedProgress(6.5)
    
            else if(Phase0State=1): #"Each Slate Bread has a higher chance of granting a specific ingredient"
                ValidBakery.HideUKQuestDescription()
                if(Camera:=AllCameraPhase0[0]):
                    spawn: 
                        PlayCamera(Camera,Agent)
                if(ValidBakery.PlayerDataCache.Language=ENGLISH):
                    ValidBakery.GameManager.DisplayCharacterMessageLoc(Agent,I_CHARACTER,BasicMessage(tutorial_en_module.Phase0_UKMessage1),?Sound:=MESSAGE_SOUND,?DisplayTime:=6.0,?Priority:=MAX_PRIORITY)
                else if(ValidBakery.PlayerDataCache.Language=SPANISH):
                    ValidBakery.GameManager.DisplayCharacterMessageLoc(Agent,I_CHARACTER,BasicMessage(tutorial_es_module.Phase0_UKMessage1),?Sound:=MESSAGE_SOUND,?DisplayTime:=6.0,?Priority:=MAX_PRIORITY)
                else if(ValidBakery.PlayerDataCache.Language=FRENCH):
                    ValidBakery.GameManager.DisplayCharacterMessageLoc(Agent,I_CHARACTER,BasicMessage(tutorial_fr_module.Phase0_UKMessage1),?Sound:=MESSAGE_SOUND,?DisplayTime:=6.0,?Priority:=MAX_PRIORITY)
                else if(ValidBakery.PlayerDataCache.Language=PORTUGUESE):
                    ValidBakery.GameManager.DisplayCharacterMessageLoc(Agent,I_CHARACTER,BasicMessage(tutorial_pt_module.Phase0_UKMessage1),?Sound:=MESSAGE_SOUND,?DisplayTime:=6.0,?Priority:=MAX_PRIORITY)
    
                spawn:
                    AddDelayedProgress(6.5)
    
            else if(Phase0State=2): #"Gather 5 Water"
                ValidBakery.HideUKQuestDescription()
                set CurrentTutorialObjetive_Type = WATER
                set CurrentTutorialProgress_Amount = 0.0
                set CurrentTutorialObjetive_Amount = 5.0
                set CurrentTutorial_Icon = Images.Water_Icon
                if(ValidBakery.PlayerDataCache.Language=ENGLISH):
                    ValidBakery.GameManager.DisplayCharacterMessageLoc(Agent,I_CHARACTER,BasicMessage(otherstrings_en_module.GatherWater),?Sound:=MESSAGE_SOUND,?DisplayTime:=6.0,?Priority:=MAX_PRIORITY)
                    ValidBakery.SetUKQuestDescription(BasicMessage(otherstrings_en_module.GatherWater))
                else if(ValidBakery.PlayerDataCache.Language=SPANISH):
                    ValidBakery.GameManager.DisplayCharacterMessageLoc(Agent,I_CHARACTER,BasicMessage(otherstrings_es_module.GatherWater),?Sound:=MESSAGE_SOUND,?DisplayTime:=6.0,?Priority:=MAX_PRIORITY)
                    ValidBakery.SetUKQuestDescription(BasicMessage(otherstrings_es_module.GatherWater))
                else if(ValidBakery.PlayerDataCache.Language=FRENCH):
                    ValidBakery.GameManager.DisplayCharacterMessageLoc(Agent,I_CHARACTER,BasicMessage(otherstrings_fr_module.GatherWater),?Sound:=MESSAGE_SOUND,?DisplayTime:=6.0,?Priority:=MAX_PRIORITY)
                    ValidBakery.SetUKQuestDescription(BasicMessage(otherstrings_fr_module.GatherWater))
                else if(ValidBakery.PlayerDataCache.Language=PORTUGUESE):
                    ValidBakery.GameManager.DisplayCharacterMessageLoc(Agent,I_CHARACTER,BasicMessage(otherstrings_pt_module.GatherWater),?Sound:=MESSAGE_SOUND,?DisplayTime:=6.0,?Priority:=MAX_PRIORITY)
                    ValidBakery.SetUKQuestDescription(BasicMessage(otherstrings_pt_module.GatherWater))
    
            else if(Phase0State=3): #"Gather 5 Flour"
                ValidBakery.HideUKQuestDescription()
                set CurrentTutorialObjetive_Type = FLOUR
                set CurrentTutorialProgress_Amount = 0.0
                set CurrentTutorialObjetive_Amount = 5.0
                set CurrentTutorial_Icon = Images.Flour_Icon
                if(ValidBakery.PlayerDataCache.Language=ENGLISH):
                    ValidBakery.GameManager.DisplayCharacterMessageLoc(Agent,I_CHARACTER,BasicMessage(otherstrings_en_module.GatherFlour),?Sound:=MESSAGE_SOUND,?DisplayTime:=6.0,?Priority:=MAX_PRIORITY)
                    ValidBakery.SetUKQuestDescription(BasicMessage(otherstrings_en_module.GatherFlour))
                else if(ValidBakery.PlayerDataCache.Language=SPANISH):
                    ValidBakery.GameManager.DisplayCharacterMessageLoc(Agent,I_CHARACTER,BasicMessage(otherstrings_es_module.GatherFlour),?Sound:=MESSAGE_SOUND,?DisplayTime:=6.0,?Priority:=MAX_PRIORITY)
                    ValidBakery.SetUKQuestDescription(BasicMessage(otherstrings_es_module.GatherFlour))
                else if(ValidBakery.PlayerDataCache.Language=FRENCH):
                    ValidBakery.GameManager.DisplayCharacterMessageLoc(Agent,I_CHARACTER,BasicMessage(otherstrings_fr_module.GatherFlour),?Sound:=MESSAGE_SOUND,?DisplayTime:=6.0,?Priority:=MAX_PRIORITY)
                    ValidBakery.SetUKQuestDescription(BasicMessage(otherstrings_fr_module.GatherFlour))
                else if(ValidBakery.PlayerDataCache.Language=PORTUGUESE):
                    ValidBakery.GameManager.DisplayCharacterMessageLoc(Agent,I_CHARACTER,BasicMessage(otherstrings_pt_module.GatherFlour),?Sound:=MESSAGE_SOUND,?DisplayTime:=6.0,?Priority:=MAX_PRIORITY)
                    ValidBakery.SetUKQuestDescription(BasicMessage(otherstrings_pt_module.GatherFlour))
    
            else if(Phase0State>=4): #"Good job"
                set CurrentTutorialObjetive_Type = -1
                ValidBakery.HideUKQuestDescription()
                if(ValidBakery.PlayerDataCache.Language=ENGLISH):
                    ValidBakery.GameManager.DisplayCharacterMessageLoc(Agent,I_CHARACTER,BasicMessage(otherstrings_en_module.GoodJob),?Sound:=BIGOBJETIVE_SOUND,?Priority:=MAX_PRIORITY)
                else if(ValidBakery.PlayerDataCache.Language=SPANISH):
                    ValidBakery.GameManager.DisplayCharacterMessageLoc(Agent,I_CHARACTER,BasicMessage(otherstrings_es_module.GoodJob),?Sound:=BIGOBJETIVE_SOUND,?Priority:=MAX_PRIORITY)
                else if(ValidBakery.PlayerDataCache.Language=FRENCH):
                    ValidBakery.GameManager.DisplayCharacterMessageLoc(Agent,I_CHARACTER,BasicMessage(otherstrings_fr_module.GoodJob),?Sound:=BIGOBJETIVE_SOUND,?Priority:=MAX_PRIORITY)
                else if(ValidBakery.PlayerDataCache.Language=PORTUGUESE):
                    ValidBakery.GameManager.DisplayCharacterMessageLoc(Agent,I_CHARACTER,BasicMessage(otherstrings_pt_module.GoodJob),?Sound:=BIGOBJETIVE_SOUND,?Priority:=MAX_PRIORITY)
                
                spawn:
                    DelayPlayPhase(1)
    

    DelayPlayPhase(NumPhase : int)<suspends> : void = 
        Sleep(5.0)
        if(NumPhase = 1):
            PlayPhase1()
        else if(NumPhase = 2):
            PlayPhase2()
        else if(NumPhase = 3):
            PlayPhase3()
        else if(NumPhase = 4):
            PlayPhase4()

    PlayPhase1() : void =
        if(ValidBakery:=MaybeBakery?, Agent := ValidBakery.MaybeAgentOwner?):
            #Buttons
            CoalButton.Disable()
            BakerButton.Disable()
            BakersInfoButton.Disable()
            Pick.Disable()
            Valve.Disable()
            Funnel.Enable()
            DroneButton.Disable()
            #teleport Player
            set CurrentPhase = 1
            Log("Phase1State: {Phase1State}",LOG_LEVEL_UK)
            if(Phase1State=0): #"Now introduce the ingredients into the funnel"
                ValidBakery.HideUKQuestDescription()
                TeleportToCurrentPhase(Agent)
                
                set CurrentTutorialObjetive_Type = USEFUNNEL_TUTORIAL_TYPE
                set CurrentTutorialProgress_Amount = 0.0
                set CurrentTutorialObjetive_Amount = 1.0
                set CurrentTutorial_Icon = Images.Challenge_Icon
            
                if(ValidBakery.PlayerDataCache.Language=ENGLISH):
                    ValidBakery.GameManager.DisplayCharacterMessageLoc(Agent, I_CHARACTER, BasicMessage(tutorial_en_module.Phase1_UKMessage0), ?Sound:=MESSAGE_SOUND, ?DisplayTime:=6.0, ?Priority:=MAX_PRIORITY)
                    ValidBakery.SetUKQuestDescription(BasicMessage(tutorial_en_module.Phase1_UKMessage1)) # Set Info
                else if(ValidBakery.PlayerDataCache.Language=SPANISH):
                    ValidBakery.GameManager.DisplayCharacterMessageLoc(Agent, I_CHARACTER, BasicMessage(tutorial_es_module.Phase1_UKMessage0), ?Sound:=MESSAGE_SOUND, ?DisplayTime:=6.0, ?Priority:=MAX_PRIORITY)
                    ValidBakery.SetUKQuestDescription(BasicMessage(tutorial_es_module.Phase1_UKMessage1)) # Set Info
                else if(ValidBakery.PlayerDataCache.Language=FRENCH):
                    ValidBakery.GameManager.DisplayCharacterMessageLoc(Agent, I_CHARACTER, BasicMessage(tutorial_fr_module.Phase1_UKMessage0), ?Sound:=MESSAGE_SOUND, ?DisplayTime:=6.0, ?Priority:=MAX_PRIORITY)
                    ValidBakery.SetUKQuestDescription(BasicMessage(tutorial_fr_module.Phase1_UKMessage1)) # Set Info
                else if(ValidBakery.PlayerDataCache.Language=PORTUGUESE):
                    ValidBakery.GameManager.DisplayCharacterMessageLoc(Agent, I_CHARACTER, BasicMessage(tutorial_pt_module.Phase1_UKMessage0), ?Sound:=MESSAGE_SOUND, ?DisplayTime:=6.0, ?Priority:=MAX_PRIORITY)
                    ValidBakery.SetUKQuestDescription(BasicMessage(tutorial_pt_module.Phase1_UKMessage1)) # Set Info
            
                if(Camera:=AllCameraPhase1[0]):
                    spawn: 
                        PlayCamera(Camera,Agent)
            
            else if(Phase1State=1): #"Check out how your main deposits have filled up"
                set CurrentTutorialObjetive_Type = -1 
                ValidBakery.HideUKQuestDescription()
            
                if(ValidBakery.PlayerDataCache.Language=ENGLISH):
                    ValidBakery.GameManager.DisplayCharacterMessageLoc(Agent, I_CHARACTER, BasicMessage(tutorial_en_module.Phase1_UKMessage2), ?Sound:=MESSAGE_SOUND, ?DisplayTime:=6.0, ?Priority:=MAX_PRIORITY)
                else if(ValidBakery.PlayerDataCache.Language=SPANISH):
                    ValidBakery.GameManager.DisplayCharacterMessageLoc(Agent, I_CHARACTER, BasicMessage(tutorial_es_module.Phase1_UKMessage2), ?Sound:=MESSAGE_SOUND, ?DisplayTime:=6.0, ?Priority:=MAX_PRIORITY)
                else if(ValidBakery.PlayerDataCache.Language=FRENCH):
                    ValidBakery.GameManager.DisplayCharacterMessageLoc(Agent, I_CHARACTER, BasicMessage(tutorial_fr_module.Phase1_UKMessage2), ?Sound:=MESSAGE_SOUND, ?DisplayTime:=6.0, ?Priority:=MAX_PRIORITY)
                else if(ValidBakery.PlayerDataCache.Language=PORTUGUESE):
                    ValidBakery.GameManager.DisplayCharacterMessageLoc(Agent, I_CHARACTER, BasicMessage(tutorial_pt_module.Phase1_UKMessage2), ?Sound:=MESSAGE_SOUND, ?DisplayTime:=6.0, ?Priority:=MAX_PRIORITY)
            
                if(Camera:=AllCameraPhase1[1]):
                    spawn: 
                        PlayCamera(Camera,Agent)
            
                spawn:
                    AddDelayedProgress(6.5)
            
            else if(Phase1State=2): #"Good job!"
                ValidBakery.HideUKQuestDescription()
            
                if(ValidBakery.PlayerDataCache.Language=ENGLISH):
                    ValidBakery.GameManager.DisplayCharacterMessageLoc(Agent, I_CHARACTER, BasicMessage(otherstrings_en_module.GoodJob), ?Sound:=BIGOBJETIVE_SOUND, ?DisplayTime:=6.0, ?Priority:=MAX_PRIORITY)
                else if(ValidBakery.PlayerDataCache.Language=SPANISH):
                    ValidBakery.GameManager.DisplayCharacterMessageLoc(Agent, I_CHARACTER, BasicMessage(otherstrings_es_module.GoodJob), ?Sound:=BIGOBJETIVE_SOUND, ?DisplayTime:=6.0, ?Priority:=MAX_PRIORITY)
                else if(ValidBakery.PlayerDataCache.Language=FRENCH):
                    ValidBakery.GameManager.DisplayCharacterMessageLoc(Agent, I_CHARACTER, BasicMessage(otherstrings_fr_module.GoodJob), ?Sound:=BIGOBJETIVE_SOUND, ?DisplayTime:=6.0, ?Priority:=MAX_PRIORITY)
                else if(ValidBakery.PlayerDataCache.Language=PORTUGUESE):
                    ValidBakery.GameManager.DisplayCharacterMessageLoc(Agent, I_CHARACTER, BasicMessage(otherstrings_pt_module.GoodJob), ?Sound:=BIGOBJETIVE_SOUND, ?DisplayTime:=6.0, ?Priority:=MAX_PRIORITY)
            
                spawn:
                    DelayPlayPhase(2)
                        
                        

    ResetBaker() : void =
        set SuperEnergy_Multiplier = 1.0
        set SuperEnergy = 0.0
        set ValveState = 0
        set MainDeposit_Water = 0.0
        set MainDeposit_Flour = 0.0
        set MainDeposit_Cheese = 0.0
        set MainDeposit_Pimentos = 0.0
        set WaterPreparedFor_B = 0.0
        set FlourPreparedFor_B = 0.0
        UpdateAllMainDeposits()
        UpdateSubdeposits()
        UpdateValveProps()
        SetSuperEnergyMultiplierProp()

    PlayValveChange(Agent : agent)<suspends> : void =
        if(Bakery:=MaybeBakery?):
            set ValveState = 1
            UpdateValveProps()
            Bakery.GameManager.VFXManager.Audio_ValveChanged.Play(Agent)
            TakeResourcersFromDeposits()
            Sleep(2.0)
            set ValveState = 2
            UpdateValveProps()
            Bakery.GameManager.VFXManager.Audio_ValveChanged.Play(Agent)
            TakeResourcersFromDeposits()

    PlayPhase2() : void = 

        <#
        Phase2_UKMessage0<public><localizes>():message = "Now let´s learn about the Bakers"
        Phase2_UKMessage1<public><localizes>():message = "Gather 5 Water"
        Phase2_UKMessage2<public><localizes>():message = "Gather 5 Flour"
        Phase2_UKMessage3<public><localizes>():message = "Introduce the ingredients into the funnel"
        Phase2_UKMessage31<public><localizes>():message = "Use the funnel"
        Phase2_UKMessage4<public><localizes>():message = "Open the valves so ingredients enter the Baker"
        Phase2_UKMessage41<public><localizes>():message = "Open the Valves"
        Phase2_UKMessage45<public><localizes>():message = "A Green valve has more priority than a Yellow one"
        Phase2_UKMessage5<public><localizes>():message = "These bars show what’s needed for the next batch, try to keep them full"
        Phase2_UKMessage6<public><localizes>():message = "You need more ingredients. Let´s gather more"
        Phase2_UKMessage61<public><localizes>():message = "Gather 5 Water"
        Phase2_UKMessage62<public><localizes>():message = "Gather 5 Flour"
        Phase2_UKMessage7<public><localizes>():message = "Good job! Now the Baker is Baking Bread"
        Phase2_UKMessage8<public><localizes>():message = "If the Baker is working on the next batch, these lights will be green"
        Phase2_UKMessage9<public><localizes>():message = "We will increase the Baker’s speed in this tutorial"
        Phase2_UKMessage10<public><localizes>():message = "Wow! That looks great!"
        Phase2_UKMessage11<public><localizes>():message = "Now, send the Baked Bread using the drone"
        Phase2_UKMessage111<public><localizes>():message = "Send the drone"
        Phase2_UKMessage12<public><localizes>():message = "Good job!"
        #>

        if(ValidBakery:=MaybeBakery?, Agent := ValidBakery.MaybeAgentOwner?):
            #teleport Player
            #Buttons
            CoalButton.Disable()
            BakerButton.Disable()
            BakersInfoButton.Disable()
            Pick.Disable()
            Valve.Enable()
            Funnel.Enable()
            DroneButton.Enable()
            set CurrentPhase = 2
            Log("Phase2State: {Phase2State}",LOG_LEVEL_UK)
            if(Phase2State=0): #"Now let’s learn about the Bakers"
                ResetBaker()
                ValidBakery.HideUKQuestDescription()
                TeleportToCurrentPhase(Agent)
            
                if(ValidBakery.PlayerDataCache.Language=ENGLISH):
                    ValidBakery.GameManager.DisplayCharacterMessageLoc(Agent, I_CHARACTER, BasicMessage(tutorial_en_module.Phase2_UKMessage0), ?Sound:=MESSAGE_SOUND, ?DisplayTime:=6.0, ?Priority:=MAX_PRIORITY)
                else if(ValidBakery.PlayerDataCache.Language=SPANISH):
                    ValidBakery.GameManager.DisplayCharacterMessageLoc(Agent, I_CHARACTER, BasicMessage(tutorial_es_module.Phase2_UKMessage0), ?Sound:=MESSAGE_SOUND, ?DisplayTime:=6.0, ?Priority:=MAX_PRIORITY)
                else if(ValidBakery.PlayerDataCache.Language=FRENCH):
                    ValidBakery.GameManager.DisplayCharacterMessageLoc(Agent, I_CHARACTER, BasicMessage(tutorial_fr_module.Phase2_UKMessage0), ?Sound:=MESSAGE_SOUND, ?DisplayTime:=6.0, ?Priority:=MAX_PRIORITY)
                else if(ValidBakery.PlayerDataCache.Language=PORTUGUESE):
                    ValidBakery.GameManager.DisplayCharacterMessageLoc(Agent, I_CHARACTER, BasicMessage(tutorial_pt_module.Phase2_UKMessage0), ?Sound:=MESSAGE_SOUND, ?DisplayTime:=6.0, ?Priority:=MAX_PRIORITY)
            
                spawn:
                    AddDelayedProgress(6.5)
            
            else if(Phase2State=1): #"Gather 5 Water"
                ValidBakery.HideUKQuestDescription()
                set CurrentTutorialObjetive_Type = WATER
                set CurrentTutorialProgress_Amount = 0.0
                set CurrentTutorialObjetive_Amount = 5.0
                set CurrentTutorial_Icon = Images.Water_Icon
            
                if(Camera:=AllCameraPhase2[0]):
                    spawn: 
                        PlayCamera(Camera,Agent)
            
                if(ValidBakery.PlayerDataCache.Language=ENGLISH):
                    ValidBakery.GameManager.DisplayCharacterMessageLoc(Agent, I_CHARACTER, BasicMessage(tutorial_en_module.Phase2_UKMessage1), ?Sound:=MESSAGE_SOUND, ?DisplayTime:=6.0, ?Priority:=MAX_PRIORITY)
                    ValidBakery.SetUKQuestDescription(BasicMessage(tutorial_en_module.Phase2_UKMessage1))
                else if(ValidBakery.PlayerDataCache.Language=SPANISH):
                    ValidBakery.GameManager.DisplayCharacterMessageLoc(Agent, I_CHARACTER, BasicMessage(tutorial_es_module.Phase2_UKMessage1), ?Sound:=MESSAGE_SOUND, ?DisplayTime:=6.0, ?Priority:=MAX_PRIORITY)
                    ValidBakery.SetUKQuestDescription(BasicMessage(tutorial_es_module.Phase2_UKMessage1))
                else if(ValidBakery.PlayerDataCache.Language=FRENCH):
                    ValidBakery.GameManager.DisplayCharacterMessageLoc(Agent, I_CHARACTER, BasicMessage(tutorial_fr_module.Phase2_UKMessage1), ?Sound:=MESSAGE_SOUND, ?DisplayTime:=6.0, ?Priority:=MAX_PRIORITY)
                    ValidBakery.SetUKQuestDescription(BasicMessage(tutorial_fr_module.Phase2_UKMessage1))
                else if(ValidBakery.PlayerDataCache.Language=PORTUGUESE):
                    ValidBakery.GameManager.DisplayCharacterMessageLoc(Agent, I_CHARACTER, BasicMessage(tutorial_pt_module.Phase2_UKMessage1), ?Sound:=MESSAGE_SOUND, ?DisplayTime:=6.0, ?Priority:=MAX_PRIORITY)
                    ValidBakery.SetUKQuestDescription(BasicMessage(tutorial_pt_module.Phase2_UKMessage1))
            
            else if(Phase2State=2): #"Gather 5 Flour"
                ValidBakery.HideUKQuestDescription()
                set CurrentTutorialObjetive_Type = FLOUR
                set CurrentTutorialProgress_Amount = 0.0
                set CurrentTutorialObjetive_Amount = 5.0
                set CurrentTutorial_Icon = Images.Flour_Icon
            
                if(ValidBakery.PlayerDataCache.Language=ENGLISH):
                    ValidBakery.GameManager.DisplayCharacterMessageLoc(Agent, I_CHARACTER, BasicMessage(tutorial_en_module.Phase2_UKMessage2), ?Sound:=MESSAGE_SOUND, ?DisplayTime:=6.0, ?Priority:=MAX_PRIORITY)
                    ValidBakery.SetUKQuestDescription(BasicMessage(tutorial_en_module.Phase2_UKMessage2))
                else if(ValidBakery.PlayerDataCache.Language=SPANISH):
                    ValidBakery.GameManager.DisplayCharacterMessageLoc(Agent, I_CHARACTER, BasicMessage(tutorial_es_module.Phase2_UKMessage2), ?Sound:=MESSAGE_SOUND, ?DisplayTime:=6.0, ?Priority:=MAX_PRIORITY)
                    ValidBakery.SetUKQuestDescription(BasicMessage(tutorial_es_module.Phase2_UKMessage2))
                else if(ValidBakery.PlayerDataCache.Language=FRENCH):
                    ValidBakery.GameManager.DisplayCharacterMessageLoc(Agent, I_CHARACTER, BasicMessage(tutorial_fr_module.Phase2_UKMessage2), ?Sound:=MESSAGE_SOUND, ?DisplayTime:=6.0, ?Priority:=MAX_PRIORITY)
                    ValidBakery.SetUKQuestDescription(BasicMessage(tutorial_fr_module.Phase2_UKMessage2))
                else if(ValidBakery.PlayerDataCache.Language=PORTUGUESE):
                    ValidBakery.GameManager.DisplayCharacterMessageLoc(Agent, I_CHARACTER, BasicMessage(tutorial_pt_module.Phase2_UKMessage2), ?Sound:=MESSAGE_SOUND, ?DisplayTime:=6.0, ?Priority:=MAX_PRIORITY)
                    ValidBakery.SetUKQuestDescription(BasicMessage(tutorial_pt_module.Phase2_UKMessage2))
            
            else if(Phase2State=3): #"Introduce the ingredients into the funnel"
                ValidBakery.HideUKQuestDescription()
                set CurrentTutorialObjetive_Type = USEFUNNEL_TUTORIAL_TYPE
                set CurrentTutorialProgress_Amount = 0.0
                set CurrentTutorialObjetive_Amount = 1.0
                set CurrentTutorial_Icon = Images.Challenge_Icon
            
                if(ValidBakery.PlayerDataCache.Language=ENGLISH):
                    ValidBakery.GameManager.DisplayCharacterMessageLoc(Agent, I_CHARACTER, BasicMessage(tutorial_en_module.Phase2_UKMessage3), ?Sound:=MESSAGE_SOUND, ?DisplayTime:=6.0, ?Priority:=MAX_PRIORITY)
                    ValidBakery.SetUKQuestDescription(BasicMessage(otherstrings_en_module.Funnel))
                else if(ValidBakery.PlayerDataCache.Language=SPANISH):
                    ValidBakery.GameManager.DisplayCharacterMessageLoc(Agent, I_CHARACTER, BasicMessage(tutorial_es_module.Phase2_UKMessage3), ?Sound:=MESSAGE_SOUND, ?DisplayTime:=6.0, ?Priority:=MAX_PRIORITY)
                    ValidBakery.SetUKQuestDescription(BasicMessage(otherstrings_es_module.Funnel))
                else if(ValidBakery.PlayerDataCache.Language=FRENCH):
                    ValidBakery.GameManager.DisplayCharacterMessageLoc(Agent, I_CHARACTER, BasicMessage(tutorial_fr_module.Phase2_UKMessage3), ?Sound:=MESSAGE_SOUND, ?DisplayTime:=6.0, ?Priority:=MAX_PRIORITY)
                    ValidBakery.SetUKQuestDescription(BasicMessage(otherstrings_fr_module.Funnel))
                else if(ValidBakery.PlayerDataCache.Language=PORTUGUESE):
                    ValidBakery.GameManager.DisplayCharacterMessageLoc(Agent, I_CHARACTER, BasicMessage(tutorial_pt_module.Phase2_UKMessage3), ?Sound:=MESSAGE_SOUND, ?DisplayTime:=6.0, ?Priority:=MAX_PRIORITY)
                    ValidBakery.SetUKQuestDescription(BasicMessage(otherstrings_pt_module.Funnel))
            


            else if(Phase2State=4): #"Open the valves so ingredients enter the Baker"
                ValidBakery.HideUKQuestDescription()
                set CurrentTutorialObjetive_Type = USEVALVES_TUTORIAL_TYPE
                set CurrentTutorialProgress_Amount = 0.0
                set CurrentTutorialObjetive_Amount = 1.0
                set CurrentTutorial_Icon = Images.Challenge_Icon

                if(Camera:=AllCameraPhase2[1]):
                    spawn: 
                        PlayCamera(Camera,Agent)

                if(ValidBakery.PlayerDataCache.Language=ENGLISH):
                    ValidBakery.GameManager.DisplayCharacterMessageLoc(Agent, I_CHARACTER, BasicMessage(tutorial_en_module.Phase2_UKMessage4), ?Sound:=MESSAGE_SOUND, ?DisplayTime:=6.0, ?Priority:=MAX_PRIORITY)
                    ValidBakery.SetUKQuestDescription(BasicMessage(otherstrings_en_module.OpenValves))
                else if(ValidBakery.PlayerDataCache.Language=SPANISH):
                    ValidBakery.GameManager.DisplayCharacterMessageLoc(Agent, I_CHARACTER, BasicMessage(tutorial_es_module.Phase2_UKMessage4), ?Sound:=MESSAGE_SOUND, ?DisplayTime:=6.0, ?Priority:=MAX_PRIORITY)
                    ValidBakery.SetUKQuestDescription(BasicMessage(otherstrings_es_module.OpenValves))
                else if(ValidBakery.PlayerDataCache.Language=FRENCH):
                    ValidBakery.GameManager.DisplayCharacterMessageLoc(Agent, I_CHARACTER, BasicMessage(tutorial_fr_module.Phase2_UKMessage4), ?Sound:=MESSAGE_SOUND, ?DisplayTime:=6.0, ?Priority:=MAX_PRIORITY)
                    ValidBakery.SetUKQuestDescription(BasicMessage(otherstrings_fr_module.OpenValves))
                else if(ValidBakery.PlayerDataCache.Language=PORTUGUESE):
                    ValidBakery.GameManager.DisplayCharacterMessageLoc(Agent, I_CHARACTER, BasicMessage(tutorial_pt_module.Phase2_UKMessage4), ?Sound:=MESSAGE_SOUND, ?DisplayTime:=6.0, ?Priority:=MAX_PRIORITY)
                    ValidBakery.SetUKQuestDescription(BasicMessage(otherstrings_pt_module.OpenValves))

            else if(Phase2State=5): #"More priority"
                set CurrentTutorialObjetive_Type = -1 
                TeleportToCurrentPhase(Agent)
                ValidBakery.HideUKQuestDescription()

                if(Camera:=AllCameraPhase2[1]):
                    spawn: 
                        PlayCamera(Camera,Agent,?CameraTime:=6.0)

                if(ValidBakery.PlayerDataCache.Language=ENGLISH):
                    ValidBakery.GameManager.DisplayCharacterMessageLoc(Agent, I_CHARACTER, BasicMessage(tutorial_en_module.Phase2_UKMessage44), ?Sound:=MESSAGE_SOUND, ?DisplayTime:=6.0, ?Priority:=MAX_PRIORITY)
                else if(ValidBakery.PlayerDataCache.Language=SPANISH):
                    ValidBakery.GameManager.DisplayCharacterMessageLoc(Agent, I_CHARACTER, BasicMessage(tutorial_es_module.Phase2_UKMessage44), ?Sound:=MESSAGE_SOUND, ?DisplayTime:=6.0, ?Priority:=MAX_PRIORITY)
                else if(ValidBakery.PlayerDataCache.Language=FRENCH):
                    ValidBakery.GameManager.DisplayCharacterMessageLoc(Agent, I_CHARACTER, BasicMessage(tutorial_fr_module.Phase2_UKMessage44), ?Sound:=MESSAGE_SOUND, ?DisplayTime:=6.0, ?Priority:=MAX_PRIORITY)
                else if(ValidBakery.PlayerDataCache.Language=PORTUGUESE):
                    ValidBakery.GameManager.DisplayCharacterMessageLoc(Agent, I_CHARACTER, BasicMessage(tutorial_pt_module.Phase2_UKMessage44), ?Sound:=MESSAGE_SOUND, ?DisplayTime:=6.0, ?Priority:=MAX_PRIORITY)

                spawn:
                    PlayValveChange(Agent)
                spawn:
                    AddDelayedProgress(6.5)

            else if(Phase2State=6): #"These Valves"
                ValidBakery.HideUKQuestDescription()

                if(Camera:=AllCameraPhase2[2]):
                    spawn: 
                        PlayCamera(Camera,Agent,?CameraTime:=6.0)

                if(ValidBakery.PlayerDataCache.Language=ENGLISH):
                    ValidBakery.GameManager.DisplayCharacterMessageLoc(Agent, I_CHARACTER, BasicMessage(tutorial_en_module.Phase2_UKMessage5), ?Sound:=MESSAGE_SOUND, ?DisplayTime:=6.0, ?Priority:=MAX_PRIORITY)
                else if(ValidBakery.PlayerDataCache.Language=SPANISH):
                    ValidBakery.GameManager.DisplayCharacterMessageLoc(Agent, I_CHARACTER, BasicMessage(tutorial_es_module.Phase2_UKMessage5), ?Sound:=MESSAGE_SOUND, ?DisplayTime:=6.0, ?Priority:=MAX_PRIORITY)
                else if(ValidBakery.PlayerDataCache.Language=FRENCH):
                    ValidBakery.GameManager.DisplayCharacterMessageLoc(Agent, I_CHARACTER, BasicMessage(tutorial_fr_module.Phase2_UKMessage5), ?Sound:=MESSAGE_SOUND, ?DisplayTime:=6.0, ?Priority:=MAX_PRIORITY)
                else if(ValidBakery.PlayerDataCache.Language=PORTUGUESE):
                    ValidBakery.GameManager.DisplayCharacterMessageLoc(Agent, I_CHARACTER, BasicMessage(tutorial_pt_module.Phase2_UKMessage5), ?Sound:=MESSAGE_SOUND, ?DisplayTime:=6.0, ?Priority:=MAX_PRIORITY)

                spawn:
                    AddDelayedProgress(6.5)

            else if(Phase2State=7): #"You need more ingredients. Let’s gather more"
                ValidBakery.HideUKQuestDescription()
            
                if(ValidBakery.PlayerDataCache.Language=ENGLISH):
                    ValidBakery.GameManager.DisplayCharacterMessageLoc(Agent, I_CHARACTER, BasicMessage(tutorial_en_module.Phase2_UKMessage6), ?Sound:=MESSAGE_SOUND, ?DisplayTime:=6.0, ?Priority:=MAX_PRIORITY)
                else if(ValidBakery.PlayerDataCache.Language=SPANISH):
                    ValidBakery.GameManager.DisplayCharacterMessageLoc(Agent, I_CHARACTER, BasicMessage(tutorial_es_module.Phase2_UKMessage6), ?Sound:=MESSAGE_SOUND, ?DisplayTime:=6.0, ?Priority:=MAX_PRIORITY)
                else if(ValidBakery.PlayerDataCache.Language=FRENCH):
                    ValidBakery.GameManager.DisplayCharacterMessageLoc(Agent, I_CHARACTER, BasicMessage(tutorial_fr_module.Phase2_UKMessage6), ?Sound:=MESSAGE_SOUND, ?DisplayTime:=6.0, ?Priority:=MAX_PRIORITY)
                else if(ValidBakery.PlayerDataCache.Language=PORTUGUESE):
                    ValidBakery.GameManager.DisplayCharacterMessageLoc(Agent, I_CHARACTER, BasicMessage(tutorial_pt_module.Phase2_UKMessage6), ?Sound:=MESSAGE_SOUND, ?DisplayTime:=6.0, ?Priority:=MAX_PRIORITY)
            
                spawn:
                    AddDelayedProgress(6.5)
            
            else if(Phase2State=8): #"Gather 5 Water"
                ValidBakery.HideUKQuestDescription()
                set CurrentTutorialObjetive_Type = WATER
                set CurrentTutorialProgress_Amount = 0.0
                set CurrentTutorialObjetive_Amount = 5.0
                set CurrentTutorial_Icon = Images.Water_Icon
            
                if(Camera:=AllCameraPhase2[0]):
                    spawn: 
                        PlayCamera(Camera,Agent)
            
                if(ValidBakery.PlayerDataCache.Language=ENGLISH):
                    ValidBakery.GameManager.DisplayCharacterMessageLoc(Agent, I_CHARACTER, BasicMessage(otherstrings_en_module.GatherWater), ?Sound:=MESSAGE_SOUND, ?DisplayTime:=6.0, ?Priority:=MAX_PRIORITY)
                    ValidBakery.SetUKQuestDescription(BasicMessage(otherstrings_en_module.GatherWater))
                else if(ValidBakery.PlayerDataCache.Language=SPANISH):
                    ValidBakery.GameManager.DisplayCharacterMessageLoc(Agent, I_CHARACTER, BasicMessage(otherstrings_es_module.GatherWater), ?Sound:=MESSAGE_SOUND, ?DisplayTime:=6.0, ?Priority:=MAX_PRIORITY)
                    ValidBakery.SetUKQuestDescription(BasicMessage(otherstrings_es_module.GatherWater))
                else if(ValidBakery.PlayerDataCache.Language=FRENCH):
                    ValidBakery.GameManager.DisplayCharacterMessageLoc(Agent, I_CHARACTER, BasicMessage(otherstrings_fr_module.GatherWater), ?Sound:=MESSAGE_SOUND, ?DisplayTime:=6.0, ?Priority:=MAX_PRIORITY)
                    ValidBakery.SetUKQuestDescription(BasicMessage(otherstrings_fr_module.GatherWater))
                else if(ValidBakery.PlayerDataCache.Language=PORTUGUESE):
                    ValidBakery.GameManager.DisplayCharacterMessageLoc(Agent, I_CHARACTER, BasicMessage(otherstrings_pt_module.GatherWater), ?Sound:=MESSAGE_SOUND, ?DisplayTime:=6.0, ?Priority:=MAX_PRIORITY)
                    ValidBakery.SetUKQuestDescription(BasicMessage(otherstrings_pt_module.GatherWater))
            
            else if(Phase2State=9): #"Gather 5 Flour"
                ValidBakery.HideUKQuestDescription()
                set CurrentTutorialObjetive_Type = FLOUR
                set CurrentTutorialProgress_Amount = 0.0
                set CurrentTutorialObjetive_Amount = 5.0
                set CurrentTutorial_Icon = Images.Flour_Icon
            
                if(ValidBakery.PlayerDataCache.Language=ENGLISH):
                    ValidBakery.SetUKQuestDescription(BasicMessage(otherstrings_en_module.GatherFlour))
                    ValidBakery.GameManager.DisplayCharacterMessageLoc(Agent, I_CHARACTER, BasicMessage(otherstrings_en_module.GatherFlour), ?Sound:=MESSAGE_SOUND, ?DisplayTime:=6.0, ?Priority:=MAX_PRIORITY)
                else if(ValidBakery.PlayerDataCache.Language=SPANISH):
                    ValidBakery.SetUKQuestDescription(BasicMessage(otherstrings_es_module.GatherFlour))
                    ValidBakery.GameManager.DisplayCharacterMessageLoc(Agent, I_CHARACTER, BasicMessage(otherstrings_es_module.GatherFlour), ?Sound:=MESSAGE_SOUND, ?DisplayTime:=6.0, ?Priority:=MAX_PRIORITY)
                else if(ValidBakery.PlayerDataCache.Language=FRENCH):
                    ValidBakery.SetUKQuestDescription(BasicMessage(otherstrings_fr_module.GatherFlour))
                    ValidBakery.GameManager.DisplayCharacterMessageLoc(Agent, I_CHARACTER, BasicMessage(otherstrings_fr_module.GatherFlour), ?Sound:=MESSAGE_SOUND, ?DisplayTime:=6.0, ?Priority:=MAX_PRIORITY)
                else if(ValidBakery.PlayerDataCache.Language=PORTUGUESE):
                    ValidBakery.SetUKQuestDescription(BasicMessage(otherstrings_pt_module.GatherFlour))
                    ValidBakery.GameManager.DisplayCharacterMessageLoc(Agent, I_CHARACTER, BasicMessage(otherstrings_pt_module.GatherFlour), ?Sound:=MESSAGE_SOUND, ?DisplayTime:=6.0, ?Priority:=MAX_PRIORITY)
            
            else if(Phase2State=10): #"Introduce in the funnel"
                ValidBakery.HideUKQuestDescription()
                set CurrentTutorialObjetive_Type = USEFUNNEL_TUTORIAL_TYPE
                set CurrentTutorialProgress_Amount = 0.0
                set CurrentTutorialObjetive_Amount = 1.0
                set CurrentTutorial_Icon = Images.Challenge_Icon
            
                if(ValidBakery.PlayerDataCache.Language=ENGLISH):
                    ValidBakery.SetUKQuestDescription(BasicMessage(otherstrings_en_module.Funnel))
                    ValidBakery.GameManager.DisplayCharacterMessageLoc(Agent, I_CHARACTER, BasicMessage(otherstrings_en_module.Funnel), ?Sound:=MESSAGE_SOUND, ?DisplayTime:=3.0, ?Priority:=MAX_PRIORITY)
                else if(ValidBakery.PlayerDataCache.Language=SPANISH):
                    ValidBakery.SetUKQuestDescription(BasicMessage(otherstrings_es_module.Funnel))
                    ValidBakery.GameManager.DisplayCharacterMessageLoc(Agent, I_CHARACTER, BasicMessage(otherstrings_es_module.Funnel), ?Sound:=MESSAGE_SOUND, ?DisplayTime:=3.0, ?Priority:=MAX_PRIORITY)
                else if(ValidBakery.PlayerDataCache.Language=FRENCH):
                    ValidBakery.SetUKQuestDescription(BasicMessage(otherstrings_fr_module.Funnel))
                    ValidBakery.GameManager.DisplayCharacterMessageLoc(Agent, I_CHARACTER, BasicMessage(otherstrings_fr_module.Funnel), ?Sound:=MESSAGE_SOUND, ?DisplayTime:=3.0, ?Priority:=MAX_PRIORITY)
                else if(ValidBakery.PlayerDataCache.Language=PORTUGUESE):
                    ValidBakery.SetUKQuestDescription(BasicMessage(otherstrings_pt_module.Funnel))
                    ValidBakery.GameManager.DisplayCharacterMessageLoc(Agent, I_CHARACTER, BasicMessage(otherstrings_pt_module.Funnel), ?Sound:=MESSAGE_SOUND, ?DisplayTime:=3.0, ?Priority:=MAX_PRIORITY)
            


            else if(Phase2State=11): #"Good job! Now the Baker is Baking Bread"
                set CurrentTutorialObjetive_Type = -1 
                Baker.SetMaterial(Materials.MI_Complement_green, ?Index:=1)
                ValidBakery.HideUKQuestDescription()
            
                if(ValidBakery.PlayerDataCache.Language=ENGLISH):
                    ValidBakery.GameManager.DisplayCharacterMessageLoc(Agent, I_CHARACTER, BasicMessage(tutorial_en_module.Phase2_UKMessage7), ?Sound:=MESSAGE_SOUND, ?DisplayTime:=6.0, ?Priority:=MAX_PRIORITY)
                else if(ValidBakery.PlayerDataCache.Language=SPANISH):
                    ValidBakery.GameManager.DisplayCharacterMessageLoc(Agent, I_CHARACTER, BasicMessage(tutorial_es_module.Phase2_UKMessage7), ?Sound:=MESSAGE_SOUND, ?DisplayTime:=6.0, ?Priority:=MAX_PRIORITY)
                else if(ValidBakery.PlayerDataCache.Language=FRENCH):
                    ValidBakery.GameManager.DisplayCharacterMessageLoc(Agent, I_CHARACTER, BasicMessage(tutorial_fr_module.Phase2_UKMessage7), ?Sound:=MESSAGE_SOUND, ?DisplayTime:=6.0, ?Priority:=MAX_PRIORITY)
                else if(ValidBakery.PlayerDataCache.Language=PORTUGUESE):
                    ValidBakery.GameManager.DisplayCharacterMessageLoc(Agent, I_CHARACTER, BasicMessage(tutorial_pt_module.Phase2_UKMessage7), ?Sound:=MESSAGE_SOUND, ?DisplayTime:=6.0, ?Priority:=MAX_PRIORITY)
            
                spawn:
                    AddDelayedProgress(6.5)
            
            else if(Phase2State=12): #"If the Baker is working on the next batch, these lights will be green"
                Baker.SetMaterial(Materials.MI_Complement_green, ?Index:=1)
                ValidBakery.HideUKQuestDescription()
            
                if(Camera:=AllCameraPhase2[3]):
                    spawn: 
                        PlayCamera(Camera,Agent,?CameraTime:=5.0)
            
                if(ValidBakery.PlayerDataCache.Language=ENGLISH):
                    ValidBakery.GameManager.DisplayCharacterMessageLoc(Agent, I_CHARACTER, BasicMessage(tutorial_en_module.Phase2_UKMessage8), ?Sound:=MESSAGE_SOUND, ?DisplayTime:=6.0, ?Priority:=MAX_PRIORITY)
                else if(ValidBakery.PlayerDataCache.Language=SPANISH):
                    ValidBakery.GameManager.DisplayCharacterMessageLoc(Agent, I_CHARACTER, BasicMessage(tutorial_es_module.Phase2_UKMessage8), ?Sound:=MESSAGE_SOUND, ?DisplayTime:=6.0, ?Priority:=MAX_PRIORITY)
                else if(ValidBakery.PlayerDataCache.Language=FRENCH):
                    ValidBakery.GameManager.DisplayCharacterMessageLoc(Agent, I_CHARACTER, BasicMessage(tutorial_fr_module.Phase2_UKMessage8), ?Sound:=MESSAGE_SOUND, ?DisplayTime:=6.0, ?Priority:=MAX_PRIORITY)
                else if(ValidBakery.PlayerDataCache.Language=PORTUGUESE):
                    ValidBakery.GameManager.DisplayCharacterMessageLoc(Agent, I_CHARACTER, BasicMessage(tutorial_pt_module.Phase2_UKMessage8), ?Sound:=MESSAGE_SOUND, ?DisplayTime:=6.0, ?Priority:=MAX_PRIORITY)
            
                spawn:
                    AddDelayedProgress(6.5)
            
            else if(Phase2State=13): #"We will increase the Baker’s speed in this tutorial"
                ValidBakery.HideUKQuestDescription()
            
                if(ValidBakery.PlayerDataCache.Language=ENGLISH):
                    ValidBakery.GameManager.DisplayCharacterMessageLoc(Agent, I_CHARACTER, BasicMessage(tutorial_en_module.Phase2_UKMessage9), ?Sound:=MESSAGE_SOUND, ?DisplayTime:=6.0, ?Priority:=MAX_PRIORITY)
                else if(ValidBakery.PlayerDataCache.Language=SPANISH):
                    ValidBakery.GameManager.DisplayCharacterMessageLoc(Agent, I_CHARACTER, BasicMessage(tutorial_es_module.Phase2_UKMessage9), ?Sound:=MESSAGE_SOUND, ?DisplayTime:=6.0, ?Priority:=MAX_PRIORITY)
                else if(ValidBakery.PlayerDataCache.Language=FRENCH):
                    ValidBakery.GameManager.DisplayCharacterMessageLoc(Agent, I_CHARACTER, BasicMessage(tutorial_fr_module.Phase2_UKMessage9), ?Sound:=MESSAGE_SOUND, ?DisplayTime:=6.0, ?Priority:=MAX_PRIORITY)
                else if(ValidBakery.PlayerDataCache.Language=PORTUGUESE):
                    ValidBakery.GameManager.DisplayCharacterMessageLoc(Agent, I_CHARACTER, BasicMessage(tutorial_pt_module.Phase2_UKMessage9), ?Sound:=MESSAGE_SOUND, ?DisplayTime:=6.0, ?Priority:=MAX_PRIORITY)
            
                spawn:
                    AddDelayedProgress(6.5)
            
            else if(Phase2State=14): #"Wow! That looks great!"
                ValidBakery.HideUKQuestDescription()
                AnimateBake()
            
                if(Camera:=AllCameraPhase2[4]):
                    spawn: 
                        PlayCamera(Camera,Agent)
            
                if(ValidBakery.PlayerDataCache.Language=ENGLISH):
                    ValidBakery.GameManager.DisplayCharacterMessageLoc(Agent, I_CHARACTER, BasicMessage(tutorial_en_module.Phase2_UKMessage10), ?Sound:=BIGOBJETIVE_SOUND, ?DisplayTime:=6.0, ?Priority:=MAX_PRIORITY)
                else if(ValidBakery.PlayerDataCache.Language=SPANISH):
                    ValidBakery.GameManager.DisplayCharacterMessageLoc(Agent, I_CHARACTER, BasicMessage(tutorial_es_module.Phase2_UKMessage10), ?Sound:=BIGOBJETIVE_SOUND, ?DisplayTime:=6.0, ?Priority:=MAX_PRIORITY)
                else if(ValidBakery.PlayerDataCache.Language=FRENCH):
                    ValidBakery.GameManager.DisplayCharacterMessageLoc(Agent, I_CHARACTER, BasicMessage(tutorial_fr_module.Phase2_UKMessage10), ?Sound:=BIGOBJETIVE_SOUND, ?DisplayTime:=6.0, ?Priority:=MAX_PRIORITY)
                else if(ValidBakery.PlayerDataCache.Language=PORTUGUESE):
                    ValidBakery.GameManager.DisplayCharacterMessageLoc(Agent, I_CHARACTER, BasicMessage(tutorial_pt_module.Phase2_UKMessage10), ?Sound:=BIGOBJETIVE_SOUND, ?DisplayTime:=6.0, ?Priority:=MAX_PRIORITY)
            
                spawn:
                    AddDelayedProgress(6.5)
            
            else if(Phase2State=15): #"Send the drone"
                Baker.SetMaterial(Materials.MI_Complement_red, ?Index:=1)
                ValidBakery.HideUKQuestDescription()
            
                if(Camera:=AllCameraPhase2[5]):
                    spawn: 
                        PlayCamera(Camera,Agent)
            
                set CurrentTutorialObjetive_Type = SENDDRONE_TUTORIAL_TYPE
                set CurrentTutorialProgress_Amount = 0.0
                set CurrentTutorialObjetive_Amount = 1.0
                set CurrentTutorial_Icon = Images.Challenge_Icon
            
                if(ValidBakery.PlayerDataCache.Language=ENGLISH):
                    ValidBakery.GameManager.DisplayCharacterMessageLoc(Agent, I_CHARACTER, BasicMessage(tutorial_en_module.Phase2_UKMessage11), ?Sound:=MESSAGE_SOUND, ?DisplayTime:=6.0, ?Priority:=MAX_PRIORITY)
                    ValidBakery.SetUKQuestDescription(BasicMessage(otherstrings_en_module.SendDrone))
                else if(ValidBakery.PlayerDataCache.Language=SPANISH):
                    ValidBakery.GameManager.DisplayCharacterMessageLoc(Agent, I_CHARACTER, BasicMessage(tutorial_es_module.Phase2_UKMessage11), ?Sound:=MESSAGE_SOUND, ?DisplayTime:=6.0, ?Priority:=MAX_PRIORITY)
                    ValidBakery.SetUKQuestDescription(BasicMessage(otherstrings_es_module.SendDrone))
                else if(ValidBakery.PlayerDataCache.Language=FRENCH):
                    ValidBakery.GameManager.DisplayCharacterMessageLoc(Agent, I_CHARACTER, BasicMessage(tutorial_fr_module.Phase2_UKMessage11), ?Sound:=MESSAGE_SOUND, ?DisplayTime:=6.0, ?Priority:=MAX_PRIORITY)
                    ValidBakery.SetUKQuestDescription(BasicMessage(otherstrings_fr_module.SendDrone))
                else if(ValidBakery.PlayerDataCache.Language=PORTUGUESE):
                    ValidBakery.GameManager.DisplayCharacterMessageLoc(Agent, I_CHARACTER, BasicMessage(tutorial_pt_module.Phase2_UKMessage11), ?Sound:=MESSAGE_SOUND, ?DisplayTime:=6.0, ?Priority:=MAX_PRIORITY)
                    ValidBakery.SetUKQuestDescription(BasicMessage(otherstrings_pt_module.SendDrone))
            
            else if(Phase2State=16): #"Good job!"
                set CurrentTutorialObjetive_Type = -1 
                ValidBakery.HideUKQuestDescription()
            
                if(Camera:=AllCameraPhase2[6]):
                    spawn: 
                        PlayCamera(Camera,Agent)
            
                spawn:
                    DelayPlayPhase(3)
                    

    PlayPhase3() : void =

        <#
        Phase3_UKMessage0<public><localizes>():message = "Let's upgrade the Pick Gather with these Sapphires"
        Phase3_UKMessage1<public><localizes>():message = "Great! Now you gather more ingredients per pick"
        Phase3_UKMessage2<public><localizes>():message = "Check your new pick damage on the Slate Bread"
        Phase3_UKMessage3<public><localizes>():message = "Cool!"
        Phase3_UKMessage4<public><localizes>():message = "Remember how these bars show what’s needed in the next batch?"
        Phase3_UKMessage5<public><localizes>():message = "You can check amounts inside the Baker on this computer"
        Phase3_UKMessage6<public><localizes>():message = "Let’s interact with it!"
        Phase3_UKMessage7<public><localizes>():message = "Good! Have you seen the amount of resources you need?"
        Phase3_UKMessage8<public><localizes>():message = "Now let’s improve Baker’s production in this computer"
        Phase3_UKMessage9<public><localizes>():message = "Keep in mind: Upgrading production INCREASES required ingredients"
        Phase3_UKMessage10<public><localizes>():message = "Good job!"
        #>
        #teleport Player
        if(ValidBakery:=MaybeBakery?, Agent := ValidBakery.MaybeAgentOwner?):
            CoalButton.Disable()
            BakerButton.Enable()
            BakersInfoButton.Enable()
            Pick.Enable()
            Valve.Enable()
            Funnel.Enable()
            DroneButton.Enable()
            set CurrentPhase = 3
            Log("Phase3State: {Phase3State}",LOG_LEVEL_UK)
            if(Phase3State=0): #"Let's upgrade the Pick Gather with these Sapphires"
                ValidBakery.HideUKQuestDescription()
                ResetBaker()
                TeleportToCurrentPhase(Agent)
            
                if(ValidBakery.PlayerDataCache.Language=ENGLISH):
                    ValidBakery.GameManager.DisplayCharacterMessageLoc(Agent, I_CHARACTER, BasicMessage(tutorial_en_module.Phase3_UKMessage0), ?Sound:=MESSAGE_SOUND, ?DisplayTime:=6.0, ?Priority:=MAX_PRIORITY)
                    ValidBakery.SetUKQuestDescription(BasicMessage(otherstrings_en_module.UpgradePick))
                else if(ValidBakery.PlayerDataCache.Language=SPANISH):
                    ValidBakery.GameManager.DisplayCharacterMessageLoc(Agent, I_CHARACTER, BasicMessage(tutorial_es_module.Phase3_UKMessage0), ?Sound:=MESSAGE_SOUND, ?DisplayTime:=6.0, ?Priority:=MAX_PRIORITY)
                    ValidBakery.SetUKQuestDescription(BasicMessage(otherstrings_es_module.UpgradePick))
                else if(ValidBakery.PlayerDataCache.Language=FRENCH):
                    ValidBakery.GameManager.DisplayCharacterMessageLoc(Agent, I_CHARACTER, BasicMessage(tutorial_fr_module.Phase3_UKMessage0), ?Sound:=MESSAGE_SOUND, ?DisplayTime:=6.0, ?Priority:=MAX_PRIORITY)
                    ValidBakery.SetUKQuestDescription(BasicMessage(otherstrings_fr_module.UpgradePick))
                else if(ValidBakery.PlayerDataCache.Language=PORTUGUESE):
                    ValidBakery.GameManager.DisplayCharacterMessageLoc(Agent, I_CHARACTER, BasicMessage(tutorial_pt_module.Phase3_UKMessage0), ?Sound:=MESSAGE_SOUND, ?DisplayTime:=6.0, ?Priority:=MAX_PRIORITY)
                    ValidBakery.SetUKQuestDescription(BasicMessage(otherstrings_pt_module.UpgradePick))
            
                if(Camera:=AllCameraPhase3[0]):
                    spawn: 
                        PlayCamera(Camera,Agent)
            
                set CurrentTutorialObjetive_Type = IMPROVEPICKAXE_TUTORIAL_TYPE
                set CurrentTutorialProgress_Amount = 0.0
                set CurrentTutorialObjetive_Amount = 1.0
                set CurrentTutorial_Icon = Images.Pickaxe_Icon
            
            else if(Phase3State=1): #"Great! Now you gather more ingredients per pick"
                ValidBakery.HideUKQuestDescription()
            
                if(ValidBakery.PlayerDataCache.Language=ENGLISH):
                    ValidBakery.GameManager.DisplayCharacterMessageLoc(Agent, I_CHARACTER, BasicMessage(tutorial_en_module.Phase3_UKMessage1), ?Sound:=MESSAGE_SOUND, ?DisplayTime:=6.0, ?Priority:=MAX_PRIORITY)
                else if(ValidBakery.PlayerDataCache.Language=SPANISH):
                    ValidBakery.GameManager.DisplayCharacterMessageLoc(Agent, I_CHARACTER, BasicMessage(tutorial_es_module.Phase3_UKMessage1), ?Sound:=MESSAGE_SOUND, ?DisplayTime:=6.0, ?Priority:=MAX_PRIORITY)
                else if(ValidBakery.PlayerDataCache.Language=FRENCH):
                    ValidBakery.GameManager.DisplayCharacterMessageLoc(Agent, I_CHARACTER, BasicMessage(tutorial_fr_module.Phase3_UKMessage1), ?Sound:=MESSAGE_SOUND, ?DisplayTime:=6.0, ?Priority:=MAX_PRIORITY)
                else if(ValidBakery.PlayerDataCache.Language=PORTUGUESE):
                    ValidBakery.GameManager.DisplayCharacterMessageLoc(Agent, I_CHARACTER, BasicMessage(tutorial_pt_module.Phase3_UKMessage1), ?Sound:=MESSAGE_SOUND, ?DisplayTime:=6.0, ?Priority:=MAX_PRIORITY)
            
                spawn:
                    AddDelayedProgress(6.5)
            
            else if(Phase3State=2): #"Check your new pick damage on the Slate Bread"
                ValidBakery.HideUKQuestDescription()
            
                if(Camera:=AllCameraPhase3[1]):
                    spawn: 
                        PlayCamera(Camera,Agent)
            
                set CurrentTutorialObjetive_Type = PICKHIT_TUTORIAL_TYPE
                set CurrentTutorialProgress_Amount = 0.0
                set CurrentTutorialObjetive_Amount = 1.0
                set CurrentTutorial_Icon = Images.Pickaxe_Icon
            
                if(ValidBakery.PlayerDataCache.Language=ENGLISH):
                    ValidBakery.GameManager.DisplayCharacterMessageLoc(Agent, I_CHARACTER, BasicMessage(tutorial_en_module.Phase3_UKMessage2), ?Sound:=MESSAGE_SOUND, ?DisplayTime:=6.0, ?Priority:=MAX_PRIORITY)
                    ValidBakery.SetUKQuestDescription(BasicMessage(otherstrings_en_module.CheckPick))
                else if(ValidBakery.PlayerDataCache.Language=SPANISH):
                    ValidBakery.GameManager.DisplayCharacterMessageLoc(Agent, I_CHARACTER, BasicMessage(tutorial_es_module.Phase3_UKMessage2), ?Sound:=MESSAGE_SOUND, ?DisplayTime:=6.0, ?Priority:=MAX_PRIORITY)
                    ValidBakery.SetUKQuestDescription(BasicMessage(otherstrings_es_module.CheckPick))
                else if(ValidBakery.PlayerDataCache.Language=FRENCH):
                    ValidBakery.GameManager.DisplayCharacterMessageLoc(Agent, I_CHARACTER, BasicMessage(tutorial_fr_module.Phase3_UKMessage2), ?Sound:=MESSAGE_SOUND, ?DisplayTime:=6.0, ?Priority:=MAX_PRIORITY)
                    ValidBakery.SetUKQuestDescription(BasicMessage(otherstrings_fr_module.CheckPick))
                else if(ValidBakery.PlayerDataCache.Language=PORTUGUESE):
                    ValidBakery.GameManager.DisplayCharacterMessageLoc(Agent, I_CHARACTER, BasicMessage(tutorial_pt_module.Phase3_UKMessage2), ?Sound:=MESSAGE_SOUND, ?DisplayTime:=6.0, ?Priority:=MAX_PRIORITY)
                    ValidBakery.SetUKQuestDescription(BasicMessage(otherstrings_pt_module.CheckPick))
            
            else if(Phase3State=3): #"Cool!"
                ValidBakery.HideUKQuestDescription()
                set CurrentTutorialObjetive_Type = -1
            
                if(ValidBakery.PlayerDataCache.Language=ENGLISH):
                    ValidBakery.GameManager.DisplayCharacterMessageLoc(Agent, I_CHARACTER, BasicMessage(tutorial_en_module.Phase3_UKMessage3), ?Sound:=MESSAGE_SOUND, ?DisplayTime:=6.0, ?Priority:=MAX_PRIORITY)
                else if(ValidBakery.PlayerDataCache.Language=SPANISH):
                    ValidBakery.GameManager.DisplayCharacterMessageLoc(Agent, I_CHARACTER, BasicMessage(tutorial_es_module.Phase3_UKMessage3), ?Sound:=MESSAGE_SOUND, ?DisplayTime:=6.0, ?Priority:=MAX_PRIORITY)
                else if(ValidBakery.PlayerDataCache.Language=FRENCH):
                    ValidBakery.GameManager.DisplayCharacterMessageLoc(Agent, I_CHARACTER, BasicMessage(tutorial_fr_module.Phase3_UKMessage3), ?Sound:=MESSAGE_SOUND, ?DisplayTime:=6.0, ?Priority:=MAX_PRIORITY)
                else if(ValidBakery.PlayerDataCache.Language=PORTUGUESE):
                    ValidBakery.GameManager.DisplayCharacterMessageLoc(Agent, I_CHARACTER, BasicMessage(tutorial_pt_module.Phase3_UKMessage3), ?Sound:=MESSAGE_SOUND, ?DisplayTime:=6.0, ?Priority:=MAX_PRIORITY)
            
                spawn:
                    AddDelayedProgress(6.5)
            


            else if(Phase3State=4): #"Remember how these bars show what’s needed in the next batch?"
                ValidBakery.HideUKQuestDescription()
            
                if(Camera:=AllCameraPhase3[2]):
                    spawn: 
                        PlayCamera(Camera,Agent)
            
                if(ValidBakery.PlayerDataCache.Language=ENGLISH):
                    ValidBakery.GameManager.DisplayCharacterMessageLoc(Agent, I_CHARACTER, BasicMessage(tutorial_en_module.Phase3_UKMessage4), ?Sound:=MESSAGE_SOUND, ?DisplayTime:=6.0, ?Priority:=MAX_PRIORITY)
                else if(ValidBakery.PlayerDataCache.Language=SPANISH):
                    ValidBakery.GameManager.DisplayCharacterMessageLoc(Agent, I_CHARACTER, BasicMessage(tutorial_es_module.Phase3_UKMessage4), ?Sound:=MESSAGE_SOUND, ?DisplayTime:=6.0, ?Priority:=MAX_PRIORITY)
                else if(ValidBakery.PlayerDataCache.Language=FRENCH):
                    ValidBakery.GameManager.DisplayCharacterMessageLoc(Agent, I_CHARACTER, BasicMessage(tutorial_fr_module.Phase3_UKMessage4), ?Sound:=MESSAGE_SOUND, ?DisplayTime:=6.0, ?Priority:=MAX_PRIORITY)
                else if(ValidBakery.PlayerDataCache.Language=PORTUGUESE):
                    ValidBakery.GameManager.DisplayCharacterMessageLoc(Agent, I_CHARACTER, BasicMessage(tutorial_pt_module.Phase3_UKMessage4), ?Sound:=MESSAGE_SOUND, ?DisplayTime:=6.0, ?Priority:=MAX_PRIORITY)
            
                spawn:
                    AddDelayedProgress(6.5)
            
            else if(Phase3State=5): #"You can check amounts inside the Baker on this computer"
                ValidBakery.HideUKQuestDescription()
            
                if(Camera:=AllCameraPhase3[3]):
                    spawn: 
                        PlayCamera(Camera,Agent)
            
                set CurrentTutorialObjetive_Type = CHECKINFO_TUTORIAL_TYPE
                set CurrentTutorialProgress_Amount = 0.0
                set CurrentTutorialObjetive_Amount = 1.0
                set CurrentTutorial_Icon = Images.Challenge_Icon
            
                if(ValidBakery.PlayerDataCache.Language=ENGLISH):
                    ValidBakery.GameManager.DisplayCharacterMessageLoc(Agent, I_CHARACTER, BasicMessage(tutorial_en_module.Phase3_UKMessage5), ?Sound:=MESSAGE_SOUND, ?DisplayTime:=6.0, ?Priority:=MAX_PRIORITY)
                    ValidBakery.SetUKQuestDescription(BasicMessage(otherstrings_en_module.CheckComputer))
                else if(ValidBakery.PlayerDataCache.Language=SPANISH):
                    ValidBakery.GameManager.DisplayCharacterMessageLoc(Agent, I_CHARACTER, BasicMessage(tutorial_es_module.Phase3_UKMessage5), ?Sound:=MESSAGE_SOUND, ?DisplayTime:=6.0, ?Priority:=MAX_PRIORITY)
                    ValidBakery.SetUKQuestDescription(BasicMessage(otherstrings_es_module.CheckComputer))
                else if(ValidBakery.PlayerDataCache.Language=FRENCH):
                    ValidBakery.GameManager.DisplayCharacterMessageLoc(Agent, I_CHARACTER, BasicMessage(tutorial_fr_module.Phase3_UKMessage5), ?Sound:=MESSAGE_SOUND, ?DisplayTime:=6.0, ?Priority:=MAX_PRIORITY)
                    ValidBakery.SetUKQuestDescription(BasicMessage(otherstrings_fr_module.CheckComputer))
                else if(ValidBakery.PlayerDataCache.Language=PORTUGUESE):
                    ValidBakery.GameManager.DisplayCharacterMessageLoc(Agent, I_CHARACTER, BasicMessage(tutorial_pt_module.Phase3_UKMessage5), ?Sound:=MESSAGE_SOUND, ?DisplayTime:=6.0, ?Priority:=MAX_PRIORITY)
                    ValidBakery.SetUKQuestDescription(BasicMessage(otherstrings_pt_module.CheckComputer))
            
            else if(Phase3State=6): #"Good! Have you seen the amount of resources you need?"
                ValidBakery.HideUKQuestDescription()
                set CurrentTutorialObjetive_Type = -1
            
                if(ValidBakery.PlayerDataCache.Language=ENGLISH):
                    ValidBakery.GameManager.DisplayCharacterMessageLoc(Agent, I_CHARACTER, BasicMessage(tutorial_en_module.Phase3_UKMessage7), ?Sound:=MESSAGE_SOUND, ?DisplayTime:=6.0, ?Priority:=MAX_PRIORITY)
                else if(ValidBakery.PlayerDataCache.Language=SPANISH):
                    ValidBakery.GameManager.DisplayCharacterMessageLoc(Agent, I_CHARACTER, BasicMessage(tutorial_es_module.Phase3_UKMessage7), ?Sound:=MESSAGE_SOUND, ?DisplayTime:=6.0, ?Priority:=MAX_PRIORITY)
                else if(ValidBakery.PlayerDataCache.Language=FRENCH):
                    ValidBakery.GameManager.DisplayCharacterMessageLoc(Agent, I_CHARACTER, BasicMessage(tutorial_fr_module.Phase3_UKMessage7), ?Sound:=MESSAGE_SOUND, ?DisplayTime:=6.0, ?Priority:=MAX_PRIORITY)
                else if(ValidBakery.PlayerDataCache.Language=PORTUGUESE):
                    ValidBakery.GameManager.DisplayCharacterMessageLoc(Agent, I_CHARACTER, BasicMessage(tutorial_pt_module.Phase3_UKMessage7), ?Sound:=MESSAGE_SOUND, ?DisplayTime:=6.0, ?Priority:=MAX_PRIORITY)
            
                spawn:
                    AddDelayedProgress(6.5)
            
            else if(Phase3State=7): #"Now let’s improve Baker’s production in this computer"
                if(Camera:=AllCameraPhase3[4]):
                    spawn: 
                        PlayCamera(Camera,Agent)
            
                ValidBakery.HideUKQuestDescription()
                set CurrentTutorialObjetive_Type = IMPROVEBAKERY_TUTORIAL_TYPE
                set CurrentTutorialProgress_Amount = 0.0
                set CurrentTutorialObjetive_Amount = 1.0
                set CurrentTutorial_Icon = Images.Challenge_Icon
            
                if(ValidBakery.PlayerDataCache.Language=ENGLISH):
                    ValidBakery.GameManager.DisplayCharacterMessageLoc(Agent, I_CHARACTER, BasicMessage(tutorial_en_module.Phase3_UKMessage8), ?Sound:=MESSAGE_SOUND, ?DisplayTime:=6.0, ?Priority:=MAX_PRIORITY)
                    ValidBakery.SetUKQuestDescription(BasicMessage(otherstrings_en_module.ImproveProduction))
                else if(ValidBakery.PlayerDataCache.Language=SPANISH):
                    ValidBakery.GameManager.DisplayCharacterMessageLoc(Agent, I_CHARACTER, BasicMessage(tutorial_es_module.Phase3_UKMessage8), ?Sound:=MESSAGE_SOUND, ?DisplayTime:=6.0, ?Priority:=MAX_PRIORITY)
                    ValidBakery.SetUKQuestDescription(BasicMessage(otherstrings_es_module.ImproveProduction))
                else if(ValidBakery.PlayerDataCache.Language=FRENCH):
                    ValidBakery.GameManager.DisplayCharacterMessageLoc(Agent, I_CHARACTER, BasicMessage(tutorial_fr_module.Phase3_UKMessage8), ?Sound:=MESSAGE_SOUND, ?DisplayTime:=6.0, ?Priority:=MAX_PRIORITY)
                    ValidBakery.SetUKQuestDescription(BasicMessage(otherstrings_fr_module.ImproveProduction))
                else if(ValidBakery.PlayerDataCache.Language=PORTUGUESE):
                    ValidBakery.GameManager.DisplayCharacterMessageLoc(Agent, I_CHARACTER, BasicMessage(tutorial_pt_module.Phase3_UKMessage8), ?Sound:=MESSAGE_SOUND, ?DisplayTime:=6.0, ?Priority:=MAX_PRIORITY)
                    ValidBakery.SetUKQuestDescription(BasicMessage(otherstrings_pt_module.ImproveProduction))
            

            else if(Phase3State=8): #"Keep in mind: Upgrading production INCREASES required ingredients"
                ValidBakery.HideUKQuestDescription()
                set CurrentTutorialObjetive_Type = -1
            
                if(ValidBakery.PlayerDataCache.Language=ENGLISH):
                    ValidBakery.GameManager.DisplayCharacterMessageLoc(Agent, I_CHARACTER, BasicMessage(tutorial_en_module.Phase3_UKMessage9), ?Sound:=MESSAGE_SOUND, ?DisplayTime:=6.0, ?Priority:=MAX_PRIORITY)
                else if(ValidBakery.PlayerDataCache.Language=SPANISH):
                    ValidBakery.GameManager.DisplayCharacterMessageLoc(Agent, I_CHARACTER, BasicMessage(tutorial_es_module.Phase3_UKMessage9), ?Sound:=MESSAGE_SOUND, ?DisplayTime:=6.0, ?Priority:=MAX_PRIORITY)
                else if(ValidBakery.PlayerDataCache.Language=FRENCH):
                    ValidBakery.GameManager.DisplayCharacterMessageLoc(Agent, I_CHARACTER, BasicMessage(tutorial_fr_module.Phase3_UKMessage9), ?Sound:=MESSAGE_SOUND, ?DisplayTime:=6.0, ?Priority:=MAX_PRIORITY)
                else if(ValidBakery.PlayerDataCache.Language=PORTUGUESE):
                    ValidBakery.GameManager.DisplayCharacterMessageLoc(Agent, I_CHARACTER, BasicMessage(tutorial_pt_module.Phase3_UKMessage9), ?Sound:=MESSAGE_SOUND, ?DisplayTime:=6.0, ?Priority:=MAX_PRIORITY)
            
                spawn:
                    AddDelayedProgress(6.5)
            
            else if(Phase3State=9): #"Good job"
                ValidBakery.HideUKQuestDescription()
            
                if(ValidBakery.PlayerDataCache.Language=ENGLISH):
                    ValidBakery.GameManager.DisplayCharacterMessageLoc(Agent, I_CHARACTER, BasicMessage(otherstrings_en_module.GoodJob), ?Sound:=BIGOBJETIVE_SOUND, ?DisplayTime:=6.0, ?Priority:=MAX_PRIORITY)
                else if(ValidBakery.PlayerDataCache.Language=SPANISH):
                    ValidBakery.GameManager.DisplayCharacterMessageLoc(Agent, I_CHARACTER, BasicMessage(otherstrings_es_module.GoodJob), ?Sound:=BIGOBJETIVE_SOUND, ?DisplayTime:=6.0, ?Priority:=MAX_PRIORITY)
                else if(ValidBakery.PlayerDataCache.Language=FRENCH):
                    ValidBakery.GameManager.DisplayCharacterMessageLoc(Agent, I_CHARACTER, BasicMessage(otherstrings_fr_module.GoodJob), ?Sound:=BIGOBJETIVE_SOUND, ?DisplayTime:=6.0, ?Priority:=MAX_PRIORITY)
                else if(ValidBakery.PlayerDataCache.Language=PORTUGUESE):
                    ValidBakery.GameManager.DisplayCharacterMessageLoc(Agent, I_CHARACTER, BasicMessage(otherstrings_pt_module.GoodJob), ?Sound:=BIGOBJETIVE_SOUND, ?DisplayTime:=6.0, ?Priority:=MAX_PRIORITY)
            
                spawn:
                    DelayPlayPhase(4)
                    

    PlayPhase4() : void = 

        <#
        Phase4_UKMessage0<public><localizes>():message = "Finally, let's use Coal to boost production"
        Phase4_UKMessage1<public><localizes>():message = "Introduce 5 Coal into the Coal funnel"
        Phase4_UKMessage2<public><localizes>():message = "This generates Super Energy for a period of time"
        Phase4_UKMessage3<public><localizes>():message = "You can check your Super Energy remaining here"
        Phase4_UKMessage4<public><localizes>():message = "Here you can see your current multiplier"
        Phase4_UKMessage5<public><localizes>():message = "Adding Coal generates extra Shared Super Energy"
        Phase4_UKMessage6<public><localizes>():message = "This boosts all other players by a small percentage"
        Phase4_UKMessage7<public><localizes>():message = "If all players have Super Energy, you get an extra boost!"
        Phase4_UKMessage8<public><localizes>():message = "Let’s apply what we have learned!"

        
        Phase4_UKMessage10<public><localizes>():message = "Introduce the ingredients into the funnel"
        Phase4_UKMessage11<public><localizes>():message = "Before opening the valves, let’s introduce Coal"
        Phase4_UKMessage12<public><localizes>():message = "Open the valves so ingredients enter the Baker"
        Phase4_UKMessage13<public><localizes>():message = "Check if you need more ingredients to start the next batch"
        Phase4_UKMessage14<public><localizes>():message = "Good job! Now the Baker is Baking Bread"
        Phase4_UKMessage15<public><localizes>():message = "Green lights show the Baker is working on the next batch"
        Phase4_UKMessage16<public><localizes>():message = "Check how the needed amount has changed in the bars"
        Phase4_UKMessage17<public><localizes>():message = "While the Baker works, the next batch is being prepared"
        Phase4_UKMessage20<public><localizes>():message = "Now, send the Baked Bread using the drone"
        Phase4_UKMessage21<public><localizes>():message = "Good job! Now you will be teleported to the Bakery!"
        #>

        if(ValidBakery:=MaybeBakery?, Agent := ValidBakery.MaybeAgentOwner?):
            CoalButton.Enable()
            BakerButton.Enable()
            BakersInfoButton.Enable()
            Pick.Enable()
            Valve.Enable()
            Funnel.Enable()
            DroneButton.Enable()
            set CurrentPhase = 4
            if(Phase4State=0): #"Finally, let's use Coal to boost production"  
                set CurrentTutorialObjetive_Type = -1   
                ValidBakery.HideUKQuestDescription()
                ResetBaker()
                TeleportToCurrentPhase(Agent) 
            
                if(ValidBakery.PlayerDataCache.Language=ENGLISH):
                    ValidBakery.GameManager.DisplayCharacterMessageLoc(Agent, I_CHARACTER, BasicMessage(tutorial_en_module.Phase4_UKMessage0), ?Sound:=MESSAGE_SOUND, ?DisplayTime:=6.0, ?Priority:=MAX_PRIORITY)
                else if(ValidBakery.PlayerDataCache.Language=SPANISH):
                    ValidBakery.GameManager.DisplayCharacterMessageLoc(Agent, I_CHARACTER, BasicMessage(tutorial_es_module.Phase4_UKMessage0), ?Sound:=MESSAGE_SOUND, ?DisplayTime:=6.0, ?Priority:=MAX_PRIORITY)
                else if(ValidBakery.PlayerDataCache.Language=FRENCH):
                    ValidBakery.GameManager.DisplayCharacterMessageLoc(Agent, I_CHARACTER, BasicMessage(tutorial_fr_module.Phase4_UKMessage0), ?Sound:=MESSAGE_SOUND, ?DisplayTime:=6.0, ?Priority:=MAX_PRIORITY)
                else if(ValidBakery.PlayerDataCache.Language=PORTUGUESE):
                    ValidBakery.GameManager.DisplayCharacterMessageLoc(Agent, I_CHARACTER, BasicMessage(tutorial_pt_module.Phase4_UKMessage0), ?Sound:=MESSAGE_SOUND, ?DisplayTime:=6.0, ?Priority:=MAX_PRIORITY)
            
                spawn:
                    AddDelayedProgress(6.5) 
            
                Log("Phase4State: {Phase4State}",LOG_LEVEL_UK)
                Log("Phase4State: {0}",LOG_LEVEL_UK)
                Log("#############",LOG_LEVEL_UK)
            
            else if(Phase4State=1): #"Introduce 30 Coal into the Coal funnel"
                ValidBakery.HideUKQuestDescription()
            
                if(Camera:=AllCameraPhase4[0]):
                    spawn: 
                        PlayCamera(Camera,Agent)
            
                set CurrentTutorialObjetive_Type = USECOAL_TUTORIAL_TYPE
                set CurrentTutorialProgress_Amount = 0.0
                set CurrentTutorialObjetive_Amount = 30.0
                set CurrentTutorial_Icon = Images.Flour_Icon
            
                if(ValidBakery.PlayerDataCache.Language=ENGLISH):
                    ValidBakery.SetUKQuestDescription(BasicMessage(otherstrings_en_module.UseCoal))
                    ValidBakery.GameManager.DisplayCharacterMessageLoc(Agent, I_CHARACTER, BasicMessage(tutorial_en_module.Phase4_UKMessage1), ?Sound:=MESSAGE_SOUND, ?DisplayTime:=6.0, ?Priority:=MAX_PRIORITY) 
                else if(ValidBakery.PlayerDataCache.Language=SPANISH):
                    ValidBakery.SetUKQuestDescription(BasicMessage(otherstrings_es_module.UseCoal))
                    ValidBakery.GameManager.DisplayCharacterMessageLoc(Agent, I_CHARACTER, BasicMessage(tutorial_es_module.Phase4_UKMessage1), ?Sound:=MESSAGE_SOUND, ?DisplayTime:=6.0, ?Priority:=MAX_PRIORITY)
                else if(ValidBakery.PlayerDataCache.Language=FRENCH):
                    ValidBakery.SetUKQuestDescription(BasicMessage(otherstrings_fr_module.UseCoal))
                    ValidBakery.GameManager.DisplayCharacterMessageLoc(Agent, I_CHARACTER, BasicMessage(tutorial_fr_module.Phase4_UKMessage1), ?Sound:=MESSAGE_SOUND, ?DisplayTime:=6.0, ?Priority:=MAX_PRIORITY)
                else if(ValidBakery.PlayerDataCache.Language=PORTUGUESE):
                    ValidBakery.SetUKQuestDescription(BasicMessage(otherstrings_pt_module.UseCoal))
                    ValidBakery.GameManager.DisplayCharacterMessageLoc(Agent, I_CHARACTER, BasicMessage(tutorial_pt_module.Phase4_UKMessage1), ?Sound:=MESSAGE_SOUND, ?DisplayTime:=6.0, ?Priority:=MAX_PRIORITY)
            
            else if(Phase4State=2): #"This generates Super Energy for a period of time" 
                ValidBakery.HideUKQuestDescription()
                set CurrentTutorialObjetive_Type = -1   
            
                if(ValidBakery.PlayerDataCache.Language=ENGLISH):
                    ValidBakery.GameManager.DisplayCharacterMessageLoc(Agent, I_CHARACTER, BasicMessage(tutorial_en_module.Phase4_UKMessage2), ?Sound:=MESSAGE_SOUND, ?DisplayTime:=6.0, ?Priority:=MAX_PRIORITY)
                else if(ValidBakery.PlayerDataCache.Language=SPANISH):
                    ValidBakery.GameManager.DisplayCharacterMessageLoc(Agent, I_CHARACTER, BasicMessage(tutorial_es_module.Phase4_UKMessage2), ?Sound:=MESSAGE_SOUND, ?DisplayTime:=6.0, ?Priority:=MAX_PRIORITY)
                else if(ValidBakery.PlayerDataCache.Language=FRENCH):
                    ValidBakery.GameManager.DisplayCharacterMessageLoc(Agent, I_CHARACTER, BasicMessage(tutorial_fr_module.Phase4_UKMessage2), ?Sound:=MESSAGE_SOUND, ?DisplayTime:=6.0, ?Priority:=MAX_PRIORITY)
                else if(ValidBakery.PlayerDataCache.Language=PORTUGUESE):
                    ValidBakery.GameManager.DisplayCharacterMessageLoc(Agent, I_CHARACTER, BasicMessage(tutorial_pt_module.Phase4_UKMessage2), ?Sound:=MESSAGE_SOUND, ?DisplayTime:=6.0, ?Priority:=MAX_PRIORITY)
            
                spawn:
                    AddDelayedProgress(6.5) 
            
            else if(Phase4State=3): #"You can check your Super Energy remaining here"   
                ValidBakery.HideUKQuestDescription()
            
                if(Camera:=AllCameraPhase4[1]):
                    spawn: 
                        PlayCamera(Camera,Agent)
            
                if(ValidBakery.PlayerDataCache.Language=ENGLISH):
                    ValidBakery.GameManager.DisplayCharacterMessageLoc(Agent, I_CHARACTER, BasicMessage(tutorial_en_module.Phase4_UKMessage3), ?Sound:=MESSAGE_SOUND, ?DisplayTime:=6.0, ?Priority:=MAX_PRIORITY)
                else if(ValidBakery.PlayerDataCache.Language=SPANISH):
                    ValidBakery.GameManager.DisplayCharacterMessageLoc(Agent, I_CHARACTER, BasicMessage(tutorial_es_module.Phase4_UKMessage3), ?Sound:=MESSAGE_SOUND, ?DisplayTime:=6.0, ?Priority:=MAX_PRIORITY)
                else if(ValidBakery.PlayerDataCache.Language=FRENCH):
                    ValidBakery.GameManager.DisplayCharacterMessageLoc(Agent, I_CHARACTER, BasicMessage(tutorial_fr_module.Phase4_UKMessage3), ?Sound:=MESSAGE_SOUND, ?DisplayTime:=6.0, ?Priority:=MAX_PRIORITY)
                else if(ValidBakery.PlayerDataCache.Language=PORTUGUESE):
                    ValidBakery.GameManager.DisplayCharacterMessageLoc(Agent, I_CHARACTER, BasicMessage(tutorial_pt_module.Phase4_UKMessage3), ?Sound:=MESSAGE_SOUND, ?DisplayTime:=6.0, ?Priority:=MAX_PRIORITY)
            
                spawn:
                    AddDelayedProgress(6.5) 
            
            else if(Phase4State=4): #"Here you can see your current multiplier"  
                ValidBakery.HideUKQuestDescription()
            
                if(Camera:=AllCameraPhase4[2]):
                    spawn: 
                        PlayCamera(Camera,Agent)
            
                if(ValidBakery.PlayerDataCache.Language=ENGLISH):
                    ValidBakery.GameManager.DisplayCharacterMessageLoc(Agent, I_CHARACTER, BasicMessage(tutorial_en_module.Phase4_UKMessage4), ?Sound:=MESSAGE_SOUND, ?DisplayTime:=6.0, ?Priority:=MAX_PRIORITY)
                else if(ValidBakery.PlayerDataCache.Language=SPANISH):
                    ValidBakery.GameManager.DisplayCharacterMessageLoc(Agent, I_CHARACTER, BasicMessage(tutorial_es_module.Phase4_UKMessage4), ?Sound:=MESSAGE_SOUND, ?DisplayTime:=6.0, ?Priority:=MAX_PRIORITY)
                else if(ValidBakery.PlayerDataCache.Language=FRENCH):
                    ValidBakery.GameManager.DisplayCharacterMessageLoc(Agent, I_CHARACTER, BasicMessage(tutorial_fr_module.Phase4_UKMessage4), ?Sound:=MESSAGE_SOUND, ?DisplayTime:=6.0, ?Priority:=MAX_PRIORITY)
                else if(ValidBakery.PlayerDataCache.Language=PORTUGUESE):
                    ValidBakery.GameManager.DisplayCharacterMessageLoc(Agent, I_CHARACTER, BasicMessage(tutorial_pt_module.Phase4_UKMessage4), ?Sound:=MESSAGE_SOUND, ?DisplayTime:=6.0, ?Priority:=MAX_PRIORITY)
            
                spawn:
                    AddDelayedProgress(6.5) 
            


            else if(Phase4State=5): #"Adding Coal generates extra Shared Super Energy" 
                ValidBakery.HideUKQuestDescription()
            
                if(Camera:=ValidBakery.GameManager.AllExtraCameras[0]):
                    spawn: 
                        PlayCamera(Camera,Agent)
            
                if(ValidBakery.PlayerDataCache.Language=ENGLISH):
                    ValidBakery.GameManager.DisplayCharacterMessageLoc(Agent, I_CHARACTER, BasicMessage(tutorial_en_module.Phase4_UKMessage5), ?Sound:=MESSAGE_SOUND, ?DisplayTime:=6.0, ?Priority:=MAX_PRIORITY)
                else if(ValidBakery.PlayerDataCache.Language=SPANISH):
                    ValidBakery.GameManager.DisplayCharacterMessageLoc(Agent, I_CHARACTER, BasicMessage(tutorial_es_module.Phase4_UKMessage5), ?Sound:=MESSAGE_SOUND, ?DisplayTime:=6.0, ?Priority:=MAX_PRIORITY)
                else if(ValidBakery.PlayerDataCache.Language=FRENCH):
                    ValidBakery.GameManager.DisplayCharacterMessageLoc(Agent, I_CHARACTER, BasicMessage(tutorial_fr_module.Phase4_UKMessage5), ?Sound:=MESSAGE_SOUND, ?DisplayTime:=6.0, ?Priority:=MAX_PRIORITY)
                else if(ValidBakery.PlayerDataCache.Language=PORTUGUESE):
                    ValidBakery.GameManager.DisplayCharacterMessageLoc(Agent, I_CHARACTER, BasicMessage(tutorial_pt_module.Phase4_UKMessage5), ?Sound:=MESSAGE_SOUND, ?DisplayTime:=6.0, ?Priority:=MAX_PRIORITY)
            
                spawn:
                    AddDelayedProgress(6.5) 
            
            else if(Phase4State=6): #"This boosts all other players by a small percentage"
                ValidBakery.HideUKQuestDescription()
            
                if(Camera:=ValidBakery.GameManager.AllExtraCameras[1]):
                    spawn: 
                        PlayCamera(Camera,Agent)
            
                if(ValidBakery.PlayerDataCache.Language=ENGLISH):
                    ValidBakery.GameManager.DisplayCharacterMessageLoc(Agent, I_CHARACTER, BasicMessage(tutorial_en_module.Phase4_UKMessage6), ?Sound:=MESSAGE_SOUND, ?DisplayTime:=6.0, ?Priority:=MAX_PRIORITY)
                else if(ValidBakery.PlayerDataCache.Language=SPANISH):
                    ValidBakery.GameManager.DisplayCharacterMessageLoc(Agent, I_CHARACTER, BasicMessage(tutorial_es_module.Phase4_UKMessage6), ?Sound:=MESSAGE_SOUND, ?DisplayTime:=6.0, ?Priority:=MAX_PRIORITY)
                else if(ValidBakery.PlayerDataCache.Language=FRENCH):
                    ValidBakery.GameManager.DisplayCharacterMessageLoc(Agent, I_CHARACTER, BasicMessage(tutorial_fr_module.Phase4_UKMessage6), ?Sound:=MESSAGE_SOUND, ?DisplayTime:=6.0, ?Priority:=MAX_PRIORITY)
                else if(ValidBakery.PlayerDataCache.Language=PORTUGUESE):
                    ValidBakery.GameManager.DisplayCharacterMessageLoc(Agent, I_CHARACTER, BasicMessage(tutorial_pt_module.Phase4_UKMessage6), ?Sound:=MESSAGE_SOUND, ?DisplayTime:=6.0, ?Priority:=MAX_PRIORITY)
            
                spawn:
                    AddDelayedProgress(6.5) 
            
            else if(Phase4State=7): #"If all players have Super Energy, you get an extra boost!"
                ValidBakery.HideUKQuestDescription()
            
                if(ValidBakery.PlayerDataCache.Language=ENGLISH):
                    ValidBakery.GameManager.DisplayCharacterMessageLoc(Agent, I_CHARACTER, BasicMessage(tutorial_en_module.Phase4_UKMessage7), ?Sound:=MESSAGE_SOUND, ?DisplayTime:=6.0, ?Priority:=MAX_PRIORITY)
                else if(ValidBakery.PlayerDataCache.Language=SPANISH):
                    ValidBakery.GameManager.DisplayCharacterMessageLoc(Agent, I_CHARACTER, BasicMessage(tutorial_es_module.Phase4_UKMessage7), ?Sound:=MESSAGE_SOUND, ?DisplayTime:=6.0, ?Priority:=MAX_PRIORITY)
                else if(ValidBakery.PlayerDataCache.Language=FRENCH):
                    ValidBakery.GameManager.DisplayCharacterMessageLoc(Agent, I_CHARACTER, BasicMessage(tutorial_fr_module.Phase4_UKMessage7), ?Sound:=MESSAGE_SOUND, ?DisplayTime:=6.0, ?Priority:=MAX_PRIORITY)
                else if(ValidBakery.PlayerDataCache.Language=PORTUGUESE):
                    ValidBakery.GameManager.DisplayCharacterMessageLoc(Agent, I_CHARACTER, BasicMessage(tutorial_pt_module.Phase4_UKMessage7), ?Sound:=MESSAGE_SOUND, ?DisplayTime:=6.0, ?Priority:=MAX_PRIORITY)
            
                spawn:
                    AddDelayedProgress(6.5) 
            
            else if(Phase4State=8): #"Let’s apply what we have learned!"
                ValidBakery.HideUKQuestDescription()
            
                if(ValidBakery.PlayerDataCache.Language=ENGLISH):
                    ValidBakery.GameManager.DisplayCharacterMessageLoc(Agent, I_CHARACTER, BasicMessage(tutorial_en_module.Phase4_UKMessage8), ?Sound:=MESSAGE_SOUND, ?DisplayTime:=6.0, ?Priority:=MAX_PRIORITY)
                else if(ValidBakery.PlayerDataCache.Language=SPANISH):
                    ValidBakery.GameManager.DisplayCharacterMessageLoc(Agent, I_CHARACTER, BasicMessage(tutorial_es_module.Phase4_UKMessage8), ?Sound:=MESSAGE_SOUND, ?DisplayTime:=6.0, ?Priority:=MAX_PRIORITY)
                else if(ValidBakery.PlayerDataCache.Language=FRENCH):
                    ValidBakery.GameManager.DisplayCharacterMessageLoc(Agent, I_CHARACTER, BasicMessage(tutorial_fr_module.Phase4_UKMessage8), ?Sound:=MESSAGE_SOUND, ?DisplayTime:=6.0, ?Priority:=MAX_PRIORITY)
                else if(ValidBakery.PlayerDataCache.Language=PORTUGUESE):
                    ValidBakery.GameManager.DisplayCharacterMessageLoc(Agent, I_CHARACTER, BasicMessage(tutorial_pt_module.Phase4_UKMessage8), ?Sound:=MESSAGE_SOUND, ?DisplayTime:=6.0, ?Priority:=MAX_PRIORITY)
            
                spawn:
                    AddDelayedProgress(6.5) 
            
            else if(Phase4State=9): #"Gather 5 Water"
                ValidBakery.HideUKQuestDescription()
                set CurrentTutorialObjetive_Type = WATER
                set CurrentTutorialProgress_Amount = 0.0
                set CurrentTutorialObjetive_Amount = 5.0
                set CurrentTutorial_Icon = Images.Water_Icon
            
                if(ValidBakery.PlayerDataCache.Language=ENGLISH):
                    ValidBakery.SetUKQuestDescription(BasicMessage(otherstrings_en_module.GatherWater))
                    ValidBakery.GameManager.DisplayCharacterMessageLoc(Agent, I_CHARACTER, BasicMessage(otherstrings_en_module.GatherWater), ?Sound:=MESSAGE_SOUND, ?DisplayTime:=6.0, ?Priority:=MAX_PRIORITY) 
                else if(ValidBakery.PlayerDataCache.Language=SPANISH):
                    ValidBakery.SetUKQuestDescription(BasicMessage(otherstrings_es_module.GatherWater))
                    ValidBakery.GameManager.DisplayCharacterMessageLoc(Agent, I_CHARACTER, BasicMessage(otherstrings_es_module.GatherWater), ?Sound:=MESSAGE_SOUND, ?DisplayTime:=6.0, ?Priority:=MAX_PRIORITY)
                else if(ValidBakery.PlayerDataCache.Language=FRENCH):
                    ValidBakery.SetUKQuestDescription(BasicMessage(otherstrings_fr_module.GatherWater))
                    ValidBakery.GameManager.DisplayCharacterMessageLoc(Agent, I_CHARACTER, BasicMessage(otherstrings_fr_module.GatherWater), ?Sound:=MESSAGE_SOUND, ?DisplayTime:=6.0, ?Priority:=MAX_PRIORITY)
                else if(ValidBakery.PlayerDataCache.Language=PORTUGUESE):
                    ValidBakery.SetUKQuestDescription(BasicMessage(otherstrings_pt_module.GatherWater))
                    ValidBakery.GameManager.DisplayCharacterMessageLoc(Agent, I_CHARACTER, BasicMessage(otherstrings_pt_module.GatherWater), ?Sound:=MESSAGE_SOUND, ?DisplayTime:=6.0, ?Priority:=MAX_PRIORITY)
            
            else if(Phase4State=10): #"Gather 5 Flour"
                ValidBakery.HideUKQuestDescription()
                set CurrentTutorialObjetive_Type = FLOUR
                set CurrentTutorialProgress_Amount = 0.0
                set CurrentTutorialObjetive_Amount = 5.0
                set CurrentTutorial_Icon = Images.Flour_Icon
            
                if(ValidBakery.PlayerDataCache.Language=ENGLISH):
                    ValidBakery.SetUKQuestDescription(BasicMessage(otherstrings_en_module.GatherFlour))
                    ValidBakery.GameManager.DisplayCharacterMessageLoc(Agent, I_CHARACTER, BasicMessage(otherstrings_en_module.GatherFlour), ?Sound:=MESSAGE_SOUND, ?DisplayTime:=6.0, ?Priority:=MAX_PRIORITY)
                else if(ValidBakery.PlayerDataCache.Language=SPANISH):
                    ValidBakery.SetUKQuestDescription(BasicMessage(otherstrings_es_module.GatherFlour))
                    ValidBakery.GameManager.DisplayCharacterMessageLoc(Agent, I_CHARACTER, BasicMessage(otherstrings_es_module.GatherFlour), ?Sound:=MESSAGE_SOUND, ?DisplayTime:=6.0, ?Priority:=MAX_PRIORITY)
                else if(ValidBakery.PlayerDataCache.Language=FRENCH):
                    ValidBakery.SetUKQuestDescription(BasicMessage(otherstrings_fr_module.GatherFlour))
                    ValidBakery.GameManager.DisplayCharacterMessageLoc(Agent, I_CHARACTER, BasicMessage(otherstrings_fr_module.GatherFlour), ?Sound:=MESSAGE_SOUND, ?DisplayTime:=6.0, ?Priority:=MAX_PRIORITY)
                else if(ValidBakery.PlayerDataCache.Language=PORTUGUESE):
                    ValidBakery.SetUKQuestDescription(BasicMessage(otherstrings_pt_module.GatherFlour))
                    ValidBakery.GameManager.DisplayCharacterMessageLoc(Agent, I_CHARACTER, BasicMessage(otherstrings_pt_module.GatherFlour), ?Sound:=MESSAGE_SOUND, ?DisplayTime:=6.0, ?Priority:=MAX_PRIORITY)
            
            else if(Phase4State=11): #"Introduce in the funnel"
                ValidBakery.HideUKQuestDescription()
                ResetBaker()
            
                set CurrentTutorialObjetive_Type = USEFUNNEL_TUTORIAL_TYPE
                set CurrentTutorialProgress_Amount = 0.0
                set CurrentTutorialObjetive_Amount = 1.0
                set CurrentTutorial_Icon = Images.Challenge_Icon
            
                if(ValidBakery.PlayerDataCache.Language=ENGLISH):
                    ValidBakery.SetUKQuestDescription(BasicMessage(otherstrings_en_module.Funnel))
                    ValidBakery.GameManager.DisplayCharacterMessageLoc(Agent, I_CHARACTER, BasicMessage(otherstrings_en_module.Funnel), ?Sound:=MESSAGE_SOUND, ?DisplayTime:=3.0, ?Priority:=MAX_PRIORITY)
                else if(ValidBakery.PlayerDataCache.Language=SPANISH):
                    ValidBakery.SetUKQuestDescription(BasicMessage(otherstrings_es_module.Funnel))
                    ValidBakery.GameManager.DisplayCharacterMessageLoc(Agent, I_CHARACTER, BasicMessage(otherstrings_es_module.Funnel), ?Sound:=MESSAGE_SOUND, ?DisplayTime:=3.0, ?Priority:=MAX_PRIORITY)
                else if(ValidBakery.PlayerDataCache.Language=FRENCH):
                    ValidBakery.SetUKQuestDescription(BasicMessage(otherstrings_fr_module.Funnel))
                    ValidBakery.GameManager.DisplayCharacterMessageLoc(Agent, I_CHARACTER, BasicMessage(otherstrings_fr_module.Funnel), ?Sound:=MESSAGE_SOUND, ?DisplayTime:=3.0, ?Priority:=MAX_PRIORITY)
                else if(ValidBakery.PlayerDataCache.Language=PORTUGUESE):
                    ValidBakery.SetUKQuestDescription(BasicMessage(otherstrings_pt_module.Funnel))
                    ValidBakery.GameManager.DisplayCharacterMessageLoc(Agent, I_CHARACTER, BasicMessage(otherstrings_pt_module.Funnel), ?Sound:=MESSAGE_SOUND, ?DisplayTime:=3.0, ?Priority:=MAX_PRIORITY)
            
            else if(Phase4State=12): #"Before opening the valves, let’s introduce Coal in the Coal Funnel"
                ValidBakery.HideUKQuestDescription()
            
                if(Camera:=AllCameraPhase4[0]):
                    spawn: 
                        PlayCamera(Camera,Agent)
            
                set CurrentTutorialObjetive_Type = USECOAL_TUTORIAL_TYPE
                set CurrentTutorialProgress_Amount = 0.0
                set CurrentTutorialObjetive_Amount = 30.0
                set CurrentTutorial_Icon = Images.Flour_Icon
            
                if(ValidBakery.PlayerDataCache.Language=ENGLISH):
                    ValidBakery.SetUKQuestDescription(BasicMessage(otherstrings_en_module.UseCoal))
                    ValidBakery.GameManager.DisplayCharacterMessageLoc(Agent, I_CHARACTER, BasicMessage(otherstrings_en_module.UseCoal), ?Sound:=MESSAGE_SOUND, ?DisplayTime:=6.0, ?Priority:=MAX_PRIORITY)
                else if(ValidBakery.PlayerDataCache.Language=SPANISH):
                    ValidBakery.SetUKQuestDescription(BasicMessage(otherstrings_es_module.UseCoal))
                    ValidBakery.GameManager.DisplayCharacterMessageLoc(Agent, I_CHARACTER, BasicMessage(otherstrings_es_module.UseCoal), ?Sound:=MESSAGE_SOUND, ?DisplayTime:=6.0, ?Priority:=MAX_PRIORITY)
                else if(ValidBakery.PlayerDataCache.Language=FRENCH):
                    ValidBakery.SetUKQuestDescription(BasicMessage(otherstrings_fr_module.UseCoal))
                    ValidBakery.GameManager.DisplayCharacterMessageLoc(Agent, I_CHARACTER, BasicMessage(otherstrings_fr_module.UseCoal), ?Sound:=MESSAGE_SOUND, ?DisplayTime:=6.0, ?Priority:=MAX_PRIORITY)
                else if(ValidBakery.PlayerDataCache.Language=PORTUGUESE):
                    ValidBakery.SetUKQuestDescription(BasicMessage(otherstrings_pt_module.UseCoal))
                    ValidBakery.GameManager.DisplayCharacterMessageLoc(Agent, I_CHARACTER, BasicMessage(otherstrings_pt_module.UseCoal), ?Sound:=MESSAGE_SOUND, ?DisplayTime:=6.0, ?Priority:=MAX_PRIORITY)
            
            else if(Phase4State=13): #"Open the valves so ingredients enter the Baker"
                ValidBakery.HideUKQuestDescription()
            
                if(Camera:=AllCameraPhase4[3]):
                    spawn: 
                        PlayCamera(Camera,Agent)
            
                set CurrentTutorialObjetive_Type = USEVALVES_TUTORIAL_TYPE
                set CurrentTutorialProgress_Amount = 0.0
                set CurrentTutorialObjetive_Amount = 1.0
                set CurrentTutorial_Icon = Images.Bread_Icon
            
                if(ValidBakery.PlayerDataCache.Language=ENGLISH):
                    ValidBakery.SetUKQuestDescription(BasicMessage(otherstrings_en_module.OpenValves))
                    ValidBakery.GameManager.DisplayCharacterMessageLoc(Agent, I_CHARACTER, BasicMessage(otherstrings_en_module.OpenValves), ?Sound:=MESSAGE_SOUND, ?DisplayTime:=6.0, ?Priority:=MAX_PRIORITY)
                else if(ValidBakery.PlayerDataCache.Language=SPANISH):
                    ValidBakery.SetUKQuestDescription(BasicMessage(otherstrings_es_module.OpenValves))
                    ValidBakery.GameManager.DisplayCharacterMessageLoc(Agent, I_CHARACTER, BasicMessage(otherstrings_es_module.OpenValves), ?Sound:=MESSAGE_SOUND, ?DisplayTime:=6.0, ?Priority:=MAX_PRIORITY)
                else if(ValidBakery.PlayerDataCache.Language=FRENCH):
                    ValidBakery.SetUKQuestDescription(BasicMessage(otherstrings_fr_module.OpenValves))
                    ValidBakery.GameManager.DisplayCharacterMessageLoc(Agent, I_CHARACTER, BasicMessage(otherstrings_fr_module.OpenValves), ?Sound:=MESSAGE_SOUND, ?DisplayTime:=6.0, ?Priority:=MAX_PRIORITY)
                else if(ValidBakery.PlayerDataCache.Language=PORTUGUESE):
                    ValidBakery.SetUKQuestDescription(BasicMessage(otherstrings_pt_module.OpenValves))
                    ValidBakery.GameManager.DisplayCharacterMessageLoc(Agent, I_CHARACTER, BasicMessage(otherstrings_pt_module.OpenValves), ?Sound:=MESSAGE_SOUND, ?DisplayTime:=6.0, ?Priority:=MAX_PRIORITY)
            


            else if(Phase4State=14): #"Check if you need more ingredients to start the next batch"
                ValidBakery.HideUKQuestDescription()
            
                if(Camera:=AllCameraPhase4[3]):
                    spawn: 
                        PlayCamera(Camera,Agent)
            
                set CurrentTutorialObjetive_Type = COMPLETEBAKE_TUTORIAL_TYPE
                set CurrentTutorialProgress_Amount = 0.0
                set CurrentTutorialObjetive_Amount = 1.0
                set CurrentTutorial_Icon = Images.Bread_Icon
            
                if(ValidBakery.PlayerDataCache.Language=ENGLISH):
                    ValidBakery.SetUKQuestDescription(BasicMessage(otherstrings_en_module.BakeBread))
                    ValidBakery.GameManager.DisplayCharacterMessageLoc(Agent, I_CHARACTER, BasicMessage(tutorial_en_module.Phase4_UKMessage13), ?Sound:=BIGOBJETIVE_SOUND, ?DisplayTime:=6.0, ?Priority:=MAX_PRIORITY)
                else if(ValidBakery.PlayerDataCache.Language=SPANISH):
                    ValidBakery.SetUKQuestDescription(BasicMessage(otherstrings_es_module.BakeBread))
                    ValidBakery.GameManager.DisplayCharacterMessageLoc(Agent, I_CHARACTER, BasicMessage(tutorial_es_module.Phase4_UKMessage13), ?Sound:=BIGOBJETIVE_SOUND, ?DisplayTime:=6.0, ?Priority:=MAX_PRIORITY)
                else if(ValidBakery.PlayerDataCache.Language=FRENCH):
                    ValidBakery.SetUKQuestDescription(BasicMessage(otherstrings_fr_module.BakeBread))
                    ValidBakery.GameManager.DisplayCharacterMessageLoc(Agent, I_CHARACTER, BasicMessage(tutorial_fr_module.Phase4_UKMessage13), ?Sound:=BIGOBJETIVE_SOUND, ?DisplayTime:=6.0, ?Priority:=MAX_PRIORITY)
                else if(ValidBakery.PlayerDataCache.Language=PORTUGUESE):
                    ValidBakery.SetUKQuestDescription(BasicMessage(otherstrings_pt_module.BakeBread))
                    ValidBakery.GameManager.DisplayCharacterMessageLoc(Agent, I_CHARACTER, BasicMessage(tutorial_pt_module.Phase4_UKMessage13), ?Sound:=BIGOBJETIVE_SOUND, ?DisplayTime:=6.0, ?Priority:=MAX_PRIORITY)
            
                spawn:
                    StartAutoBake()
            
            else if(Phase4State=15): #"Good job! Now the Baker is Baking Bread"
                Baker.SetMaterial(Materials.MI_Complement_green, ?Index:=1)
                ValidBakery.HideUKQuestDescription()
                set CurrentTutorialObjetive_Type = -1
            
                if(Camera:=AllCameraPhase4[4]):
                    spawn: 
                        PlayCamera(Camera,Agent)
            
                if(ValidBakery.PlayerDataCache.Language=ENGLISH):
                    ValidBakery.GameManager.DisplayCharacterMessageLoc(Agent, I_CHARACTER, BasicMessage(tutorial_en_module.Phase4_UKMessage14), ?Sound:=MESSAGE_SOUND, ?DisplayTime:=6.0, ?Priority:=MAX_PRIORITY)
                else if(ValidBakery.PlayerDataCache.Language=SPANISH):
                    ValidBakery.GameManager.DisplayCharacterMessageLoc(Agent, I_CHARACTER, BasicMessage(tutorial_es_module.Phase4_UKMessage14), ?Sound:=MESSAGE_SOUND, ?DisplayTime:=6.0, ?Priority:=MAX_PRIORITY)
                else if(ValidBakery.PlayerDataCache.Language=FRENCH):
                    ValidBakery.GameManager.DisplayCharacterMessageLoc(Agent, I_CHARACTER, BasicMessage(tutorial_fr_module.Phase4_UKMessage14), ?Sound:=MESSAGE_SOUND, ?DisplayTime:=6.0, ?Priority:=MAX_PRIORITY)
                else if(ValidBakery.PlayerDataCache.Language=PORTUGUESE):
                    ValidBakery.GameManager.DisplayCharacterMessageLoc(Agent, I_CHARACTER, BasicMessage(tutorial_pt_module.Phase4_UKMessage14), ?Sound:=MESSAGE_SOUND, ?DisplayTime:=6.0, ?Priority:=MAX_PRIORITY)
            
                spawn:
                    AddDelayedProgress(6.5)
            
            else if(Phase4State=16): #"Check how the needed amount has changed in the bars"
                ValidBakery.HideUKQuestDescription()
            
                if(ValidBakery.PlayerDataCache.Language=ENGLISH):
                    ValidBakery.GameManager.DisplayCharacterMessageLoc(Agent, I_CHARACTER, BasicMessage(tutorial_en_module.Phase4_UKMessage16), ?Sound:=MESSAGE_SOUND, ?DisplayTime:=6.0, ?Priority:=MAX_PRIORITY)
                else if(ValidBakery.PlayerDataCache.Language=SPANISH):
                    ValidBakery.GameManager.DisplayCharacterMessageLoc(Agent, I_CHARACTER, BasicMessage(tutorial_es_module.Phase4_UKMessage16), ?Sound:=MESSAGE_SOUND, ?DisplayTime:=6.0, ?Priority:=MAX_PRIORITY)
                else if(ValidBakery.PlayerDataCache.Language=FRENCH):
                    ValidBakery.GameManager.DisplayCharacterMessageLoc(Agent, I_CHARACTER, BasicMessage(tutorial_fr_module.Phase4_UKMessage16), ?Sound:=MESSAGE_SOUND, ?DisplayTime:=6.0, ?Priority:=MAX_PRIORITY)
                else if(ValidBakery.PlayerDataCache.Language=PORTUGUESE):
                    ValidBakery.GameManager.DisplayCharacterMessageLoc(Agent, I_CHARACTER, BasicMessage(tutorial_pt_module.Phase4_UKMessage16), ?Sound:=MESSAGE_SOUND, ?DisplayTime:=6.0, ?Priority:=MAX_PRIORITY)
            
                spawn:
                    AddDelayedProgress(6.5)
            
            else if(Phase4State=17): #"While the Baker works, the next batch is being prepared"
                ValidBakery.HideUKQuestDescription()
            
                if(ValidBakery.PlayerDataCache.Language=ENGLISH):
                    ValidBakery.GameManager.DisplayCharacterMessageLoc(Agent, I_CHARACTER, BasicMessage(tutorial_en_module.Phase4_UKMessage17), ?Sound:=MESSAGE_SOUND, ?DisplayTime:=6.0, ?Priority:=MAX_PRIORITY)
                else if(ValidBakery.PlayerDataCache.Language=SPANISH):
                    ValidBakery.GameManager.DisplayCharacterMessageLoc(Agent, I_CHARACTER, BasicMessage(tutorial_es_module.Phase4_UKMessage17), ?Sound:=MESSAGE_SOUND, ?DisplayTime:=6.0, ?Priority:=MAX_PRIORITY)
                else if(ValidBakery.PlayerDataCache.Language=FRENCH):
                    ValidBakery.GameManager.DisplayCharacterMessageLoc(Agent, I_CHARACTER, BasicMessage(tutorial_fr_module.Phase4_UKMessage17), ?Sound:=MESSAGE_SOUND, ?DisplayTime:=6.0, ?Priority:=MAX_PRIORITY)
                else if(ValidBakery.PlayerDataCache.Language=PORTUGUESE):
                    ValidBakery.GameManager.DisplayCharacterMessageLoc(Agent, I_CHARACTER, BasicMessage(tutorial_pt_module.Phase4_UKMessage17), ?Sound:=MESSAGE_SOUND, ?DisplayTime:=6.0, ?Priority:=MAX_PRIORITY)
            
                spawn:
                    AddDelayedProgress(6.5)


                    
            else if(Phase4State=18): #"Now, send the Baked Bread using the drone"
                ValidBakery.HideUKQuestDescription()
            
                if(Camera:=AllCameraPhase4[5]):
                    spawn: 
                        PlayCamera(Camera,Agent)
            
                set CurrentTutorialObjetive_Type = SENDDRONE_TUTORIAL_TYPE
                set CurrentTutorialProgress_Amount = 0.0
                set CurrentTutorialObjetive_Amount = 1.0
                set CurrentTutorial_Icon = Images.Challenge_Icon
            
                if(ValidBakery.PlayerDataCache.Language=ENGLISH):
                    ValidBakery.SetUKQuestDescription(BasicMessage(otherstrings_en_module.SendDrone))
                    ValidBakery.GameManager.DisplayCharacterMessageLoc(Agent, I_CHARACTER, BasicMessage(otherstrings_en_module.SendDrone), ?Sound:=MESSAGE_SOUND, ?DisplayTime:=6.0, ?Priority:=MAX_PRIORITY)
                else if(ValidBakery.PlayerDataCache.Language=SPANISH):
                    ValidBakery.SetUKQuestDescription(BasicMessage(otherstrings_es_module.SendDrone))
                    ValidBakery.GameManager.DisplayCharacterMessageLoc(Agent, I_CHARACTER, BasicMessage(otherstrings_es_module.SendDrone), ?Sound:=MESSAGE_SOUND, ?DisplayTime:=6.0, ?Priority:=MAX_PRIORITY)
                else if(ValidBakery.PlayerDataCache.Language=FRENCH):
                    ValidBakery.SetUKQuestDescription(BasicMessage(otherstrings_fr_module.SendDrone))
                    ValidBakery.GameManager.DisplayCharacterMessageLoc(Agent, I_CHARACTER, BasicMessage(otherstrings_fr_module.SendDrone), ?Sound:=MESSAGE_SOUND, ?DisplayTime:=6.0, ?Priority:=MAX_PRIORITY)
                else if(ValidBakery.PlayerDataCache.Language=PORTUGUESE):
                    ValidBakery.SetUKQuestDescription(BasicMessage(otherstrings_pt_module.SendDrone))
                    ValidBakery.GameManager.DisplayCharacterMessageLoc(Agent, I_CHARACTER, BasicMessage(otherstrings_pt_module.SendDrone), ?Sound:=MESSAGE_SOUND, ?DisplayTime:=6.0, ?Priority:=MAX_PRIORITY)
            
            else if(Phase4State=19): #"Good job! Now you will be teleported to the Bakery!"
                ValidBakery.HideUKQuestDescription()
                set CurrentTutorialObjetive_Type = -1
            
                if(ValidBakery.PlayerDataCache.Language=ENGLISH):
                    ValidBakery.GameManager.DisplayCharacterMessageLoc(Agent, I_CHARACTER, BasicMessage(tutorial_en_module.Phase4_UKMessage21), ?Sound:=MESSAGE_SOUND, ?DisplayTime:=3.0, ?Priority:=MAX_PRIORITY)
                else if(ValidBakery.PlayerDataCache.Language=SPANISH):
                    ValidBakery.GameManager.DisplayCharacterMessageLoc(Agent, I_CHARACTER, BasicMessage(tutorial_es_module.Phase4_UKMessage21), ?Sound:=MESSAGE_SOUND, ?DisplayTime:=3.0, ?Priority:=MAX_PRIORITY)
                else if(ValidBakery.PlayerDataCache.Language=FRENCH):
                    ValidBakery.GameManager.DisplayCharacterMessageLoc(Agent, I_CHARACTER, BasicMessage(tutorial_fr_module.Phase4_UKMessage21), ?Sound:=MESSAGE_SOUND, ?DisplayTime:=3.0, ?Priority:=MAX_PRIORITY)
                else if(ValidBakery.PlayerDataCache.Language=PORTUGUESE):
                    ValidBakery.GameManager.DisplayCharacterMessageLoc(Agent, I_CHARACTER, BasicMessage(tutorial_pt_module.Phase4_UKMessage21), ?Sound:=MESSAGE_SOUND, ?DisplayTime:=3.0, ?Priority:=MAX_PRIORITY)
            
                spawn:
                    AddDelayedProgress(3.0)
            
            else if(Phase4State=20): #"Finish tutorial"
                ValidBakery.SetUKTutorial(false)
                ValidBakery.HideUKQuestDescription()
                FinishTutorial(false)
                    
                
    FinishTutorial(IsForced : logic) : void =
        if(ValidBakery:=MaybeBakery?, Agent := ValidBakery.MaybeAgentOwner?):
            ValidBakery.SetUKTutorial(false)
            NewData:=ValidBakery.PlayersDataManager.SetTutorialState(ValidBakery.PlayerDataCache,true)
            ValidBakery.SetNewPlayerDataCache(NewData)
            TransformAssigned := AdaptTransformFromVisualEmpty(ValidBakery.GameManager.TutorialExitPosition.GetTransform(),VisualEmptyHideZ)
            ValidBakery.GameManager.TeleporterManager.TeleportPlayerUsingTeleporter(Agent,TransformAssigned)
            if(not IsForced?):
                spawn:
                    ValidBakery.GameManager.PlayEndTutorial(Agent)
            else:
                ValidBakery.GameManager.ClassTeamSelector_PickMode.ChangeClass(Agent)
                ValidBakery.RefreshAllUI()

    StartAutoBake()<suspends> : void =
        if(ValidBakery:=MaybeBakery?, Agent := ValidBakery.MaybeAgentOwner?):
            var ElapsedTime : float = 0.0
            Print("Starting autoBake")
            loop:
                Sleep(0.25)
                if(Phase4State<>14 or not ValidBakery.IsOnUKTutorial?):
                    break
                else:
                    set ElapsedTime = ElapsedTime + 0.25
                if(ElapsedTime > 120.0):
                    if(ValidBakery.PlayerDataCache.Language=ENGLISH):
                        ValidBakery.GameManager.DisplayCharacterMessageLoc(Agent, I_CHARACTER, BasicMessage(otherstrings_en_module.Problem), ?Sound:=MESSAGE_SOUND, ?DisplayTime:=6.0, ?Priority:=MAX_PRIORITY)
                    else if(ValidBakery.PlayerDataCache.Language=SPANISH):
                        ValidBakery.GameManager.DisplayCharacterMessageLoc(Agent, I_CHARACTER, BasicMessage(otherstrings_es_module.Problem), ?Sound:=MESSAGE_SOUND, ?DisplayTime:=6.0, ?Priority:=MAX_PRIORITY)
                    else if(ValidBakery.PlayerDataCache.Language=FRENCH):
                        ValidBakery.GameManager.DisplayCharacterMessageLoc(Agent, I_CHARACTER, BasicMessage(otherstrings_fr_module.Problem), ?Sound:=MESSAGE_SOUND, ?DisplayTime:=6.0, ?Priority:=MAX_PRIORITY)
                    else if(ValidBakery.PlayerDataCache.Language=PORTUGUESE):
                        ValidBakery.GameManager.DisplayCharacterMessageLoc(Agent, I_CHARACTER, BasicMessage(otherstrings_pt_module.Problem), ?Sound:=MESSAGE_SOUND, ?DisplayTime:=6.0, ?Priority:=MAX_PRIORITY)
                
                    AddPhaseProgress()
                    
                if(WaterPreparedFor_B >= WaterReqFor_B, FlourPreparedFor_B >= FlourReqFor_B):
                    set WaterPreparedFor_B = 0.0
                    set FlourPreparedFor_B = 0.0
                    AnimateBake()
                    BakeCompletedEvent()
                    UpdateBakery()

    AddPhaseProgress() : void =
        if(CurrentPhase=0):
            set Phase0State = Phase0State + 1
            PlayPhase0()
        else if(CurrentPhase=1):
            set Phase1State = Phase1State + 1
            PlayPhase1()
        else if(CurrentPhase=2):
            set Phase2State = Phase2State + 1
            PlayPhase2()
        else if(CurrentPhase=3):
            set Phase3State = Phase3State + 1
            PlayPhase3()
        else if(CurrentPhase=4):
            set Phase4State = Phase4State + 1
            PlayPhase4()

    AddDelayedProgress(Delay : float)<suspends> : void =
        Sleep(Delay)
        if(CurrentPhase=0):
            set Phase0State = Phase0State + 1
            PlayPhase0()
        else if(CurrentPhase=1):
            set Phase1State = Phase1State + 1
            PlayPhase1()
        else if(CurrentPhase=2):
            set Phase2State = Phase2State + 1
            PlayPhase2()
        else if(CurrentPhase=3):
            set Phase3State = Phase3State + 1
            PlayPhase3()
        else if(CurrentPhase=4):
            Log("Delayed Phase4State, adding to {Phase4State}+1 ",LOG_LEVEL_UK)
            set Phase4State = Phase4State + 1
            Log("Now Delayed Phase4State, adding is {Phase4State}",LOG_LEVEL_UK)
            PlayPhase4()  
            
    TryExitTutorial(Agent : agent) : void =
        if(Bakery:=MaybeBakery?):
            if(Bakery.PlayerDataCache.IsTutorialCompleted?):
                FinishTutorial(true)
            else:
                set ConfirmationToExit = true
    
                if(Bakery.PlayerDataCache.Language=ENGLISH):
                    Bakery.GameManager.DisplayCharacterMessageLoc(Agent, I_CHARACTER, BasicMessage(otherstrings_en_module.Recomendation), ?Sound:=ERROR_SOUND, ?DisplayTime:=6.0, ?Priority:=MAX_PRIORITY)
                else if(Bakery.PlayerDataCache.Language=SPANISH):
                    Bakery.GameManager.DisplayCharacterMessageLoc(Agent, I_CHARACTER, BasicMessage(otherstrings_es_module.Recomendation), ?Sound:=ERROR_SOUND, ?DisplayTime:=6.0, ?Priority:=MAX_PRIORITY)
                else if(Bakery.PlayerDataCache.Language=FRENCH):
                    Bakery.GameManager.DisplayCharacterMessageLoc(Agent, I_CHARACTER, BasicMessage(otherstrings_fr_module.Recomendation), ?Sound:=ERROR_SOUND, ?DisplayTime:=6.0, ?Priority:=MAX_PRIORITY)
                else if(Bakery.PlayerDataCache.Language=PORTUGUESE):
                    Bakery.GameManager.DisplayCharacterMessageLoc(Agent, I_CHARACTER, BasicMessage(otherstrings_pt_module.Recomendation), ?Sound:=ERROR_SOUND, ?DisplayTime:=6.0, ?Priority:=MAX_PRIORITY)
            