
using { /Fortnite.com/Devices }
using { /Verse.org/Simulation }
using { /UnrealEngine.com/Temporary/Diagnostics }

# See https://dev.epicgames.com/documentation/en-us/uefn/create-your-own-device-in-verse for how to create a verse device.

# A Verse-authored creative device that can be placed in a level
upgrades_manager := class(creative_device):

    #Improve water input
    @editable
    var AllMutZone_Upgrade_LV_WaterInput : []mutator_zone_device = array{}
    @editable
    var AllProps_UpgradeBase_LV_WaterInput : []creative_prop = array{}

    #Improve flour input
    @editable
    var AllMutZone_Upgrade_LV_FlourInput : []mutator_zone_device = array{}
    @editable
    var AllProps_UpgradeBase_LV_FlourInput : []creative_prop = array{}

    #Improve cheese input
    @editable
    var AllMutZone_Upgrade_LV_CheeseInput : []mutator_zone_device = array{}
    @editable
    var AllProps_UpgradeBase_LV_CheeseInput : []creative_prop = array{}

    #Improve pimento input
    @editable
    var AllMutZone_Upgrade_LV_PimentoInput : []mutator_zone_device = array{}
    @editable
    var AllProps_UpgradeBase_LV_PimentoInput : []creative_prop = array{}

    #Improve BreadBaker
    @editable
    var AllMutZone_Upgrade_LV_BreadBaker : []mutator_zone_device = array{}
    @editable
    var AllProps_UpgradeBase_LV_BreadBaker : []creative_prop = array{}

    #Improve CheeseBreadBaker 
    @editable
    var AllMutZone_Upgrade_LV_CheeseBreadBaker : []mutator_zone_device = array{}
    @editable
    var AllProps_UpgradeBase_LV_CheeseBreadBaker : []creative_prop = array{}


    #Improve SlicedBreadBaker
    @editable
    var AllMutZone_Upgrade_LV_SlicedBreadBaker : []mutator_zone_device = array{}
    @editable
    var AllProps_UpgradeBase_LV_SlicedBreadBaker : []creative_prop = array{}

    #Improve CheeseBlender
    @editable
    var AllMutZone_Upgrade_LV_CheeseBlender : []mutator_zone_device = array{}
    @editable
    var AllProps_UpgradeBase_LV_CheeseBlender : []creative_prop = array{}

    #Improve PimentoCheeseBlender
    @editable
    var AllMutZone_Upgrade_LV_PimentoCheeseBlender : []mutator_zone_device = array{}
    @editable
    var AllProps_UpgradeBase_LV_PimentoCheeseBlender : []creative_prop = array{}

    GetUpgradeLVMutatorZones(Index : int) : []mutator_zone_device =
        var MutatorZoneArray : []mutator_zone_device = array{}
        if:
            UpgradeLV_WaterInput := AllMutZone_Upgrade_LV_WaterInput[Index]
            UpgradeLV_FlourInput := AllMutZone_Upgrade_LV_FlourInput[Index]
            UpgradeLV_CheeseInput := AllMutZone_Upgrade_LV_CheeseInput[Index]
            UpgradeLV_PimentoInput := AllMutZone_Upgrade_LV_PimentoInput[Index]
            UpgradeLV_BreadBaker := AllMutZone_Upgrade_LV_BreadBaker[Index]
            UpgradeLV_CheeseBreadBaker := AllMutZone_Upgrade_LV_CheeseBreadBaker[Index]
            UpgradeLV_SlicedBreadBaker := AllMutZone_Upgrade_LV_SlicedBreadBaker[Index]
            UpgradeLV_CheeseBlender := AllMutZone_Upgrade_LV_CheeseBlender[Index]
            UpgradeLV_PimentoCheeseBlender := AllMutZone_Upgrade_LV_PimentoCheeseBlender[Index]
            set MutatorZoneArray = MutatorZoneArray + array{
                UpgradeLV_WaterInput,
                UpgradeLV_FlourInput,
                UpgradeLV_CheeseInput, 
                UpgradeLV_PimentoInput, 
                UpgradeLV_BreadBaker,
                UpgradeLV_CheeseBreadBaker,
                UpgradeLV_SlicedBreadBaker,
                UpgradeLV_CheeseBlender,
                UpgradeLV_PimentoCheeseBlender}
        else:
            #Print("FATAL ERROR!: GetUpgradeLVMutatorZones(): Failed on assigning MutatorZones of Index: {Index}")
        return MutatorZoneArray

    GetUpgradeLVProps(Index : int) : []creative_prop =
        var PropsArray : []creative_prop = array{}
        if:
            UpgradeLV_WaterInput := AllProps_UpgradeBase_LV_WaterInput[Index]
            UpgradeLV_FlourInput := AllProps_UpgradeBase_LV_FlourInput[Index]
            UpgradeLV_CheeseInput := AllProps_UpgradeBase_LV_CheeseInput[Index]
            UpgradeLV_PimentoInput := AllProps_UpgradeBase_LV_PimentoInput[Index]
            UpgradeLV_BreadBaker := AllProps_UpgradeBase_LV_BreadBaker[Index]
            UpgradeLV_CheeseBreadBaker := AllProps_UpgradeBase_LV_CheeseBreadBaker[Index]
            UpgradeLV_SlicedBreadBaker := AllProps_UpgradeBase_LV_SlicedBreadBaker[Index]
            UpgradeLV_CheeseBlender := AllProps_UpgradeBase_LV_CheeseBlender[Index]
            UpgradeLV_PimentoCheeseBlender := AllProps_UpgradeBase_LV_PimentoCheeseBlender[Index]
            set PropsArray = PropsArray + array{
                UpgradeLV_WaterInput,
                UpgradeLV_FlourInput,
                UpgradeLV_CheeseInput, 
                UpgradeLV_PimentoInput, 
                UpgradeLV_BreadBaker,
                UpgradeLV_CheeseBreadBaker,
                UpgradeLV_SlicedBreadBaker,
                UpgradeLV_CheeseBlender,
                UpgradeLV_PimentoCheeseBlender}
        else:
            #Print("FATAL ERROR!: GetUpgradeLVProps(): Failed on assigning Prop ProgressBars of Index: {Index}")
        return PropsArray


    