using { /Verse.org/Simulation }
using { /Verse.org/Colors/NamedColors }
using { /Fortnite.com/Characters }
using { /Fortnite.com/Devices }
using { /UnrealEngine.com/Temporary/SpatialMath }
using { /Verse.org/Random }

MessageBasic<localizes>(InString : string)<transacts>:message = "{InString}"
Message_CheeseBreadBakerComputerInfo<localizes>(ValveState1 : string ,ValveState2 : string,ValveState3 : string,FlowValve1 : int,FlowValve2 : int,FlowValve3 : int,Production : int)<transacts>:message = "💧 Valve: {ValveState1} | Flow: {FlowValve1} units/s\n🌾 Valve: {ValveState1} | Flow: {FlowValve2} units/s\n🧀 Valve: {ValveState3} | Flow: {FlowValve3} units/s\n\n👩🏻‍🍳 Production: {Production}% per second"
GetAgentName<localizes>(Agent:agent)<computes>: message = "{Agent} is asking for help in a dungeon!"
GetPlayerName<localizes>(Player:player)<computes>: message = "{Player}"

VisualEmptyHideZ : float = 1024.0
UpVisualEmptyHideZ : float = 1034.0

#LOG LEVELS
LOG_LEVEL_BAKERS : int = 0
LOG_LEVEL_1DETAILING : int = 1
LOG_LEVEL_UIs : int = 10
LOG_LEVEL_MAP : int = 20
LOG_LEVEL_NIAGGARA : int = 15
LOG_LEVEL_PICKAXEBUG : int = -15
LOG_LEVEL_HEALTHBUG : int = -14
LOG_LEVEL_NOT4 : int = 20
LOG_LEVEL_NOT5 : int = 30
LOG_LEVEL_NOT6 : int = 100
LOG_LEVEL_NPCDEBUG : int = 101
LOG_LEVEL_BAKERDEBUG : int = 102
LOG_LEVEL_COLLECTIBLES : int = 103
LOG_LEVEL_COAL : int = 104
LOG_LEVEL_TUTORIALQUESTS : int = 105
LOG_LEVEL_RANDOMMOB : int = 105
LOG_LEVEL_DEBUG0 : int = 106
LOG_LEVEL_HB2 : int = 107
LOG_LEVEL_MINIGAMES : int = 108
LOG_LEVEL_DEBUG1 : int = 109
LOG_LEVEL_DEBUG2 : int = 110
LOG_LEVEL_DEBUG3 : int = 111
LOG_LEVEL_DEBUG4 : int = 112
LOG_LEVEL_DEBUG5 : int = 113
LOG_LEVEL_DEBUG6 : int = 114
LOG_LEVEL_FATAL : int = 115
LOG_LEVEL_DEBUG7 : int = 116
LOG_LEVEL_DEBUG8 : int = 117
LOG_LEVEL_DEBUG9 : int = 118
LOG_LEVEL_DEBUG10 : int = 119
LOG_LEVEL_LAUNCHED : int = 120
LOG_LEVEL_UK : int = 122
LOG_LEVEL_DEGRADATION : int = 123
LOG_LEVEL : int = LOG_LEVEL_DEGRADATION

Log(Text : string, Level : int):void=
    if(Level = LOG_LEVEL):
        Print(Text)

PrintFatalError(Text : string):void=
    Print("#########\n#FATAL ERROR#: " + Text, ?Color:=Red)
    
PrintWarning(Text : string):void=
    Print("WARNING: " + Text, ?Color:=Orange)

PrintLaunchWarning(Text : string):void=
    Print("LAUNCH WARNING!: " + Text, ?Color:=Violet)

IsAgentOnArray(AgentToFind : agent, AllAgentsArrayToCheck : []agent) : int =
    var ArrayPosition : int = -1
    for(Index:=0..AllAgentsArrayToCheck.Length-1, Agent := AllAgentsArrayToCheck[Index], Agent = AgentToFind):
        return Index
    return ArrayPosition


SortFloatArray(Array: []float) : []float =
    var OutPutArray : []float = Array
    n := OutPutArray.Length
    for (I := 0..n - 1):
        for (J := 0..n - I - 2):
            if(OutPutArray[J] > OutPutArray[J + 1], Temp := OutPutArray[J], ToMove := OutPutArray[J + 1],set OutPutArray[J] = ToMove, set OutPutArray[J + 1] = Temp):
    return OutPutArray

#PCSB Info
RemoveAgentFromPCSBInfoMap(ExampleMap:[agent]pcsb_info_ui, Agent:agent):[agent]pcsb_info_ui=
    var NewMap:[agent]pcsb_info_ui = map{}
    for (Key -> Value : ExampleMap, Key <> Agent):
        set NewMap = ConcatenateMaps(NewMap, map{Key => Value})
    return NewMap

#ActionInfo
RemoveAgentFromActionInfoUIMap(ExampleMap:[agent]action_info_ui, Agent:agent):[agent]action_info_ui=
    var NewMap:[agent]action_info_ui = map{}
    for (Key -> Value : ExampleMap, Key <> Agent):
        set NewMap = ConcatenateMaps(NewMap, map{Key => Value})
    return NewMap

#Entity Health
RemoveAgentFromEntityHealthUIMap(ExampleMap:[agent]entity_health_bar_ui, Agent:agent):[agent]entity_health_bar_ui=
    var NewMap:[agent]entity_health_bar_ui = map{}
    for (Key -> Value : ExampleMap, Key <> Agent):
        set NewMap = ConcatenateMaps(NewMap, map{Key => Value})
    return NewMap
        
#MainInventory
RemoveAgentFromMainInventoryUIMap(ExampleMap:[agent]main_inventory_ui, Agent:agent):[agent]main_inventory_ui=
    var NewMap:[agent]main_inventory_ui = map{}
    for (Key -> Value : ExampleMap, Key <> Agent):
        set NewMap = ConcatenateMaps(NewMap, map{Key => Value})
    return NewMap

RemoveAgentFromPickaxeUpgradesUIMap(ExampleMap:[agent]pickaxe_upgrades_ui, Agent:agent):[agent]pickaxe_upgrades_ui=
    var NewMap:[agent]pickaxe_upgrades_ui = map{}
    for (Key -> Value : ExampleMap, Key <> Agent):
        set NewMap = ConcatenateMaps(NewMap, map{Key => Value})
    return NewMap

RemoveAgentFromMessagesUIMap(ExampleMap:[agent]messages_ui, Agent:agent):[agent]messages_ui=
    var NewMap:[agent]messages_ui = map{}
    for (Key -> Value : ExampleMap, Key <> Agent):
        set NewMap = ConcatenateMaps(NewMap, map{Key => Value})
    return NewMap

RemoveAgentFromResourcersUIMap(ExampleMap:[agent]resourcers_ui, Agent:agent):[agent]resourcers_ui=
    var NewMap:[agent]resourcers_ui = map{}
    for (Key -> Value : ExampleMap, Key <> Agent):
        set NewMap = ConcatenateMaps(NewMap, map{Key => Value})
    return NewMap

RemoveAgentFromBackpackUpgradesUIMap(ExampleMap:[agent]backpack_upgrades_ui, Agent:agent):[agent]backpack_upgrades_ui=
    var NewMap:[agent]backpack_upgrades_ui = map{}
    for (Key -> Value : ExampleMap, Key <> Agent):
        set NewMap = ConcatenateMaps(NewMap, map{Key => Value})
    return NewMap
#PCS
RemoveAgentFromBakerUpgradesUIMap(ExampleMap:[agent]PCSbaker_upgrades_ui, Agent:agent):[agent]PCSbaker_upgrades_ui=
    var NewMap:[agent]PCSbaker_upgrades_ui = map{}
    for (Key -> Value : ExampleMap, Key <> Agent):
        set NewMap = ConcatenateMaps(NewMap, map{Key => Value})
    return NewMap

RemoveAgentFromDepositsUpgradesUIMap(ExampleMap:[agent]deposits_upgrades_ui, Agent:agent):[agent]deposits_upgrades_ui=
    var NewMap:[agent]deposits_upgrades_ui = map{}
    for (Key -> Value : ExampleMap, Key <> Agent):
        set NewMap = ConcatenateMaps(NewMap, map{Key => Value})
    return NewMap

RemoveAgentFromDroneUpgradesUIMap(ExampleMap:[agent]drone_upgrades_ui, Agent:agent):[agent]drone_upgrades_ui=
    var NewMap:[agent]drone_upgrades_ui = map{}
    for (Key -> Value : ExampleMap, Key <> Agent):
        set NewMap = ConcatenateMaps(NewMap, map{Key => Value})
    return NewMap

RemoveAgentFromActionUIMap(ExampleMap:[agent]action_info_ui, Agent:agent):[agent]action_info_ui=
    var NewMap:[agent]action_info_ui = map{}
    # Concatenate Keys from ExampleMap into NewMap, excluding ElementToRemove
    for (Key -> Value : ExampleMap, Key <> Agent):
        set NewMap = ConcatenateMaps(NewMap, map{Key => Value})
    return NewMap

RemoveAgentAssignedToBakery(ExampleMap:[agent]bakery_machines_manager, Agent:agent):[agent]bakery_machines_manager=
    var NewMap:[agent]bakery_machines_manager = map{}
    for (Key -> Value : ExampleMap, Key <> Agent):
        set NewMap = ConcatenateMaps(NewMap, map{Key => Value})
    return NewMap

RemoveAgentFromFloatMap(ExampleMap:[agent]float, Agent:agent):[agent]float=
    var NewMap:[agent]float = map{}
    for (Key -> Value : ExampleMap, Key <> Agent):
        set NewMap = ConcatenateMaps(NewMap, map{Key => Value})
    return NewMap


RemoveAgentFromIntMap(ExampleMap:[agent]int, Agent:agent):[agent]int=
    var NewMap:[agent]int = map{}
    for (Key -> Value : ExampleMap, Key <> Agent):
        set NewMap = ConcatenateMaps(NewMap, map{Key => Value})
    return NewMap

CalculateGeneratedAmmount_FormulaA(TLV_AutoWater : []int, BaseAmmount : float) : float =
    var Ammount : float = BaseAmmount

    for(Tier := 0..TLV_AutoWater.Length-1, LVTier := TLV_AutoWater[Tier]):
        TierAmmount := BaseAmmount
        set Ammount = Ammount + (BaseAmmount*LVTier*Tier)
    #Print("Formulas to do")
    return Ammount

#Truncate settings
FindCharIndex(SearchChar:char, Str:[]char)<transacts><decides>:int=
    var FoundIndex:int = 0
    var Found:logic = false
    for (X -> Char : Str, not Found?):
        if (Char = SearchChar):
            set FoundIndex = X
    FoundIndex

OriginalToTruncatedString(Number:float, Decimals:int):[]char=
    var Str:[]char = ToString(Number)
    if:
        DotIndex := FindCharIndex['.', Str]
        StopIndex := if (Decimals > 0) then Min(DotIndex+Decimals+1,Str.Length) else DotIndex
        Value:=Str.Slice[0, StopIndex] 
        set Str = Value
    Str

ToTruncatedFloat(Number:float,UseDecimals:int):[]char=
    var Str:[]char = ToString(Number)
    if:
        DotIndex := FindCharIndex['.', Str]
    then:
        if:
            StopIndex := if (UseDecimals > 0) then Min(DotIndex+UseDecimals+1,Str.Length) else DotIndex
            Value:=Str.Slice[0, StopIndex] 
            set Str = Value 
    Str

ToTruncatedString(Number:float, Decimals:int):[]char=
    var UseDecimals : int = 2
    var Str:[]char = ToString(Number)
    if:
        DotIndex := FindCharIndex['.', Str]
    then:
        if(DotIndex=1):
            set UseDecimals = 2
        if(DotIndex=2):
            set UseDecimals = 1
        if(DotIndex>=3):
            set UseDecimals = 0
        if:
            StopIndex := if (UseDecimals > 0) then Min(DotIndex+UseDecimals+1,Str.Length) else DotIndex
            Value:=Str.Slice[0, StopIndex] 
            set Str = Value 
    Str

GetCharMagnitudeLetter(FloatAmmount : float) : []char =
    var FinalString : []char = ToString(FloatAmmount)
    if(FloatAmmount < Pow(10.0,30.0)): #Quetta
        if(FloatAmmount < Pow(10.0,27.0)): #Ronna
            if(FloatAmmount < Pow(10.0,24.0)): #Yotta
                if(FloatAmmount < Pow(10.0,21.0)): #Zetta
                    if(FloatAmmount < Pow(10.0,18.0)): #Exa
                        if(FloatAmmount < Pow(10.0,15.0)): #Peta
                            if(FloatAmmount < Pow(10.0,12.0)): #Trillions
                                if(FloatAmmount < Pow(10.0,9.0)): #Billions
                                    if(FloatAmmount < Pow(10.0,6.0)): #Millions
                                        if(FloatAmmount < Pow(10.0,3.0)): #Kilos
                                            set FinalString = ToTruncatedString(FloatAmmount,1)
                                        else:
                                                Value:=(FloatAmmount * 1.0) / Pow(10.0,3.0)
                                                String := ToTruncatedString(Value,1)
                                                set FinalString = String + array{' ','k'}
                                    else:
                                        Value:=(FloatAmmount * 1.0) / Pow(10.0,6.0)
                                        String := ToTruncatedString(Value,1)
                                        set FinalString = String + array{' ','M'}
                                else:
                                    Value:=(FloatAmmount * 1.0) / Pow(10.0,9.0)
                                    String := ToTruncatedString(Value,1)
                                    set FinalString = String + array{' ','G'}
                            else:
                                Value:=(FloatAmmount * 1.0) / Pow(10.0,12.0)
                                String := ToTruncatedString(Value,1)
                                set FinalString = String + array{' ','T'}
                        else:
                            Value:=(FloatAmmount * 1.0) / Pow(10.0,15.0)
                            String := ToTruncatedString(Value,1)
                            set FinalString = String + array{' ','P'}
                    else:
                        Value:=(FloatAmmount * 1.0) / Pow(10.0,18.0)
                        String := ToTruncatedString(Value,1)
                        set FinalString = String + array{' ','E'}
                else:
                    Value:=(FloatAmmount * 1.0) / Pow(10.0,21.0)
                    String := ToTruncatedString(Value,1)
                    set FinalString = String + array{' ','Z'}
            else:
                Value:=(FloatAmmount * 1.0) / Pow(10.0,24.0)
                String := ToTruncatedString(Value,1)
                set FinalString = String + array{' ','Y'}
        else:
            Value:=(FloatAmmount * 1.0) / Pow(10.0,27.0)
            String := ToTruncatedString(Value,1)
            set FinalString = String + array{' ','R'}
    else:
        Value:=(FloatAmmount * 1.0) / Pow(10.0,30.0)
        String := ToTruncatedString(Value,1)
        set FinalString = String + array{' ','Q'}
    return FinalString

GetOptimizedMagnitudeLetter(FloatAmmount : float) : string =
    var FinalString : string = ""
    var ModFloatAmmount : float = FloatAmmount
    if(FloatAmmount<0.0):
        set ModFloatAmmount = -ModFloatAmmount
    if(ModFloatAmmount < Pow(10.0,30.0)): #Quetta
        if(ModFloatAmmount < Pow(10.0,27.0)): #Ronna
            if(ModFloatAmmount < Pow(10.0,24.0)): #Yotta
                if(ModFloatAmmount < Pow(10.0,21.0)): #Zetta
                    if(ModFloatAmmount < Pow(10.0,18.0)): #Exa
                        if(ModFloatAmmount < Pow(10.0,15.0)): #Peta
                            if(ModFloatAmmount < Pow(10.0,12.0)): #Trillions
                                if(ModFloatAmmount < Pow(10.0,9.0)): #Billions
                                    if(ModFloatAmmount < Pow(10.0,6.0)): #Millions
                                        if(ModFloatAmmount < Pow(10.0,3.0)): #Kilos
                                            set FinalString = ToTruncatedString(ModFloatAmmount,1)
                                        else:
                                                Value:=(ModFloatAmmount * 1.0) / Pow(10.0,3.0)
                                                String := ToTruncatedString(Value,1)
                                                set FinalString = String + " k"
                                    else:
                                        Value:=(ModFloatAmmount * 1.0) / Pow(10.0,6.0)
                                        String := ToTruncatedString(Value,1)
                                        set FinalString = String + " M"
                                else:
                                    Value:=(ModFloatAmmount * 1.0) / Pow(10.0,9.0)
                                    String := ToTruncatedString(Value,1)
                                    set FinalString = String + " G"
                            else:
                                Value:=(ModFloatAmmount * 1.0) / Pow(10.0,12.0)
                                String := ToTruncatedString(Value,1)
                                set FinalString = String + " T"
                        else:
                            Value:=(ModFloatAmmount * 1.0) / Pow(10.0,15.0)
                            String := ToTruncatedString(Value,1)
                            set FinalString = String + " P"
                    else:
                        Value:=(ModFloatAmmount * 1.0) / Pow(10.0,18.0)
                        String := ToTruncatedString(Value,1)
                        set FinalString = String + " E"
                else:
                    Value:=(ModFloatAmmount * 1.0) / Pow(10.0,21.0)
                    String := ToTruncatedString(Value,1)
                    set FinalString = String + " Z"
            else:
                Value:=(ModFloatAmmount * 1.0) / Pow(10.0,24.0)
                String := ToTruncatedString(Value,1)
                set FinalString = String + " Y"
        else:
            Value:=(ModFloatAmmount * 1.0) / Pow(10.0,27.0)
            String := ToTruncatedString(Value,1)
            set FinalString = String + " R"
    else:
        Value:=(ModFloatAmmount * 1.0) / Pow(10.0,30.0)
        String := ToTruncatedString(Value,1)
        set FinalString = String + " Q"
    if(FloatAmmount<0.0):
        set FinalString = "-" + FinalString
    return FinalString

GetMagnitudeLetter(FloatAmmount : float, PreCharacter : string, PostCharacter : string) : string =
    var FinalString : string = ""
    var ModFloatAmmount : float = FloatAmmount
    if(FloatAmmount<0.0):
        set ModFloatAmmount = -ModFloatAmmount
    if(ModFloatAmmount < Pow(10.0,30.0)): #QUETTA
        if(ModFloatAmmount < Pow(10.0,27.0)): #RONNA
            if(ModFloatAmmount < Pow(10.0,24.0)): #Yotta
                if(ModFloatAmmount < Pow(10.0,21.0)): #Zetta
                    if(ModFloatAmmount < Pow(10.0,18.0)): #Exa
                        if(ModFloatAmmount < Pow(10.0,15.0)): #Peta
                            if(ModFloatAmmount < Pow(10.0,12.0)): #Trillions
                                if(ModFloatAmmount < Pow(10.0,9.0)): #Billions
                                    if(ModFloatAmmount < Pow(10.0,6.0)): #Millions
                                        if(ModFloatAmmount < Pow(10.0,3.0)): #Kilos
                                            set FinalString = PreCharacter + ToTruncatedString(ModFloatAmmount,1) + PostCharacter
                                        else:
                                                Value:=(ModFloatAmmount * 1.0) / Pow(10.0,3.0)
                                                String := ToTruncatedString(Value,1)
                                                set FinalString = PreCharacter + String + " k"  + PostCharacter
                                    else:
                                        Value:=(ModFloatAmmount * 1.0) / Pow(10.0,6.0)
                                        String := ToTruncatedString(Value,1)
                                        set FinalString = PreCharacter + String + " M"  + PostCharacter
                                else:
                                    Value:=(ModFloatAmmount * 1.0) / Pow(10.0,9.0)
                                    String := ToTruncatedString(Value,1)
                                    set FinalString = PreCharacter + String + " G"  + PostCharacter
                            else:
                                Value:=(ModFloatAmmount * 1.0) / Pow(10.0,12.0)
                                String := ToTruncatedString(Value,1)
                                set FinalString = PreCharacter + String + " T"  + PostCharacter
                        else:
                            Value:=(ModFloatAmmount * 1.0) / Pow(10.0,15.0)
                            String := ToTruncatedString(Value,1)
                            set FinalString = PreCharacter + String + " P"  + PostCharacter
                    else:
                        Value:=(ModFloatAmmount * 1.0) / Pow(10.0,18.0)
                        String := ToTruncatedString(Value,1)
                        set FinalString = PreCharacter + String + " E"  + PostCharacter
                else:
                    Value:=(ModFloatAmmount * 1.0) / Pow(10.0,21.0)
                    String := ToTruncatedString(Value,1)
                    set FinalString = PreCharacter + String + " Z"  + PostCharacter
            else:
                Value:=(ModFloatAmmount * 1.0) / Pow(10.0,24.0)
                String := ToTruncatedString(Value,1)
                set FinalString = PreCharacter + String + " Y"  + PostCharacter
        else:
            Value:=(ModFloatAmmount * 1.0) / Pow(10.0,27.0)
            String := ToTruncatedString(Value,1)
            set FinalString = PreCharacter + String + " R"  + PostCharacter
    else:
        Value:=(ModFloatAmmount * 1.0) / Pow(10.0,30.0)
        String := ToTruncatedString(Value,1)
        set FinalString = PreCharacter + String + " Q"  + PostCharacter
    if(FloatAmmount<0.0):
        set FinalString = "-" + FinalString
    return FinalString

AdaptSeconds(Number:int):[]char=
    var Str:[]char = ToString(Number)
    if(Str.Length=1):
        set Str = "0"+Str
        return Str
    return Str
   
TransformToMinutes_Seconds(TotalSeconds : int) : tuple([]char,[]char) =
    var Minutes : []char = "00"
    var Seconds : []char = "00"

    FloatMinutes:=(TotalSeconds *1.0) / (60*1.0)
    if(IntMinutes:=Floor[FloatMinutes]):
        set Minutes = AdaptSeconds(IntMinutes)
    if(IntSeconds:=Mod[TotalSeconds,60]):
        set Seconds = AdaptSeconds(IntSeconds)
    return(Minutes,Seconds)

TeleportFailedAttempt(MaybeCharacter : ?fort_character, MaybeNPCDevice : ?npc_spawner_device, MaybeProp : ?creative_prop, Transform : transform, MaxDispersion : float) : tuple(logic,transform) =
    MaxTries : int = 5
    var NumTries : int = 0
    var TeleportSucess : logic = false
    var TeleportSucessDestiny : transform = transform{}

    loop:
        XDispersion:=GetRandomFloat(-MaxDispersion,MaxDispersion)
        YDispersion:=GetRandomFloat(-MaxDispersion,MaxDispersion)
        Destination:vector3=vector3{X:=Transform.Translation.X+XDispersion,Y:=Transform.Translation.Y+YDispersion,Z:=Transform.Translation.Z}
        #Prop
        if(Prop:=MaybeProp?):
            if(Prop.TeleportTo[Destination,Transform.Rotation]):
                set TeleportSucess = true
                set TeleportSucessDestiny = transform{Translation:=Destination,Rotation:=Transform.Rotation,Scale:=NeutralScale}
                break
            else:
                #Print("Failed Prop teleportation attempt to {Transform.Translation.X+XDispersion},{Transform.Translation.Y+YDispersion},{Transform.Translation.Z}")
                set NumTries = NumTries + 1
            if(NumTries>MaxTries):
                set TeleportSucess = false
                break
        #Device
        if(Device:=MaybeNPCDevice?):
            if(Device.TeleportTo[Destination,Transform.Rotation]):
                set TeleportSucess = true
                set TeleportSucessDestiny = transform{Translation:=Destination,Rotation:=Transform.Rotation,Scale:=NeutralScale}
                break
            else:
                #Print("Failed Device teleportation attempt to {Transform.Translation.X+XDispersion},{Transform.Translation.Y+YDispersion},{Transform.Translation.Z}")
                set NumTries = NumTries + 1
            if(NumTries>MaxTries):
                set TeleportSucess = false
                break
        #Character
        if(Character:=MaybeCharacter?, Character.IsActive[], Agent:=Character.GetAgent[], Player := player[Agent], Player.IsActive[]):
            if(Character.TeleportTo[Destination,Transform.Rotation]):
                set TeleportSucess = true
                set TeleportSucessDestiny = transform{Translation:=Destination,Rotation:=Transform.Rotation,Scale:=NeutralScale}
                break
            else:
                #Print("Failed Character teleportation attempt to {Transform.Translation.X+XDispersion},{Transform.Translation.Y+YDispersion},{Transform.Translation.Z}")
                set NumTries = NumTries + 1
            if(NumTries>MaxTries):
                set TeleportSucess = false
                break
    
    return (TeleportSucess,TeleportSucessDestiny)

AdaptTransformFromVisualEmpty(Transform : transform, VisualEmptyZ : float) : transform =
    var OutPutTransform : transform = Transform
    NewTranslation := vector3{X:=Transform.Translation.X,Y:=Transform.Translation.Y,Z:=Transform.Translation.Z+VisualEmptyZ}
    NewTransform := transform{Translation:=NewTranslation,Rotation:=Transform.Rotation,Scale:=Transform.Scale}
    return NewTransform

AdaptTranslationFromVisualEmpty(Translation : vector3, VisualEmptyZ : float) : vector3 =
    var OutPutTranslation : vector3 = Translation
    NewTranslation := vector3{X:=Translation.X,Y:=Translation.Y,Z:=Translation.Z+VisualEmptyZ}
    return NewTranslation

ReleaseFromCamera(Agent : agent, Camera : gameplay_camera_fixed_point_device, DelayTime : float)<suspends> : void =
    Sleep(DelayTime)
    Camera.RemoveFrom(Agent)

PlayCamera(Camera : gameplay_camera_fixed_point_device, Agent : agent, ?CameraTime : float = 3.0)<suspends> : void =
    Camera.AddTo(Agent)
    if(Character := Agent.GetFortCharacter[]):
        Character.PutInStasis(stasis_args{AllowTurning:=false})
        ReleaseFromCamera(Agent,Camera,CameraTime)
        Character.ReleaseFromStasis()

IsOdd(X:int)<decides><transacts>:void = (Mod[X, 2] <> 0)
IsEven(X:int)<decides><transacts>:void = (Mod[X, 2] = 0)