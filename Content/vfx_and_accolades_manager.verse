
using { /Fortnite.com/Devices }
using { /Verse.org/Simulation }
using { /UnrealEngine.com/Temporary/Diagnostics }
using { /UnrealEngine.com/Temporary/SpatialMath }
using { /Fortnite.com/Devices/CreativeAnimation }
using { /Fortnite.com/Devices/CreativeAnimation/InterpolationTypes }
using { /Verse.org/Assets }
using { /Fortnite.com/Characters }

# See https://dev.epicgames.com/documentation/en-us/uefn/create-your-own-device-in-verse for how to create a verse device.

# A Verse-authored creative device that can be placed in a level
vfx_and_accolades_manager := class(creative_device):

    @editable
    AllSteveBakers : []creative_prop = array{}
    var LastSteveBake : float = 0.0

    @editable
    AllTeleporterDevices : []teleporter_device = array{}

    @editable
    PetBlenderProp : creative_prop = creative_prop{}
    @editable
    Camera_PetCreation : gameplay_camera_fixed_point_device = gameplay_camera_fixed_point_device{}

    #Quest tracker
    @editable
    QuestTrackerEffect : creative_prop_asset = DefaultCreativePropAsset
    var AllAgentsLastTranslation : []vector3 = array{vector3{X:=0.0,Y:=0.0,Z:=0.0},vector3{X:=0.0,Y:=0.0,Z:=0.0},vector3{X:=0.0,Y:=0.0,Z:=0.0},vector3{X:=0.0,Y:=0.0,Z:=0.0}}
    var AllAgentsLastTrackerTick : []float = array{0.0,0.0,0.0,0.0}

    #Super energy effects
    @editable
    CoalGenerator_Audio : audio_player_device = audio_player_device{}
    @editable
    CoalGeneratorProp_SuperEnergy : creative_prop_asset = DefaultCreativePropAsset
    @editable
    BakerGeneratorProp_SuperEnergy : creative_prop_asset = DefaultCreativePropAsset

    #Accolade devices
    @editable
    Accolade_Hit : accolades_device = accolades_device{}
    @editable
    Accolade_WeakPoint : accolades_device = accolades_device{}
    @editable
    Accolade_Mastery : accolades_device = accolades_device{}
    @editable
    Accolade_BigObjetive : accolades_device = accolades_device{}
    @editable
    Accolade_PickaxeImprovement : accolades_device = accolades_device{}
    @editable
    Accolade_BakerImprovement : accolades_device = accolades_device{}
    @editable
    Accolade_DroneSend : accolades_device = accolades_device{}
    @editable
    Accolade_HardBreadDestroyed : accolades_device = accolades_device{}
    @editable
    Accolade_NewArea : accolades_device = accolades_device{}
    @editable
    Accolade_Inventory : accolades_device = accolades_device{}
    @editable
    Accolade_AgedSapphires : accolades_device = accolades_device{}

    #Effects
    @editable
    RangeAttackEffect : creative_prop_asset = DefaultCreativePropAsset
    @editable
    RangeV2AttackEffect : creative_prop_asset = DefaultCreativePropAsset
    @editable
    MeleeAttackEffect : creative_prop_asset = DefaultCreativePropAsset
    @editable
    DefendEffect : creative_prop_asset = DefaultCreativePropAsset

    @editable
    HealAndShieldEffect : creative_prop_asset = DefaultCreativePropAsset
    
    #AUDIOS
    #Characters
    @editable
    WCharacter_AudioPlayer : audio_player_device = audio_player_device{}
    @editable
    GCharacter_AudioPlayer : audio_player_device = audio_player_device{}
    @editable
    JCharacter_AudioPlayer : audio_player_device = audio_player_device{}
    @editable
    HCharacter_AudioPlayer : audio_player_device = audio_player_device{}
    #Musics
    @editable
    HexaGameMusic_AudioPlayer : audio_player_device = audio_player_device{}
    @editable
    DungeonMusic_AudioPlayer : audio_player_device = audio_player_device{}
    @editable
    DungeonCombatMusic_AudioPlayer : audio_player_device = audio_player_device{}
    @editable
    BakersMusic_AudioPlayer : audio_player_device = audio_player_device{}
    @editable
    DesertMusic_AudioPlayer : audio_player_device = audio_player_device{}
    @editable
    SecretMusic_AudioPlayer : audio_player_device = audio_player_device{}

    #Resourcers
    @editable
    Audio_CoalGranted : audio_player_device = audio_player_device{}
    @editable
    Audio_AgedSapphiresGranted : audio_player_device = audio_player_device{}
    @editable
    Audio_MutantMaterialGranted : audio_player_device = audio_player_device{}
    @editable
    Audio_SapphiresGranted : audio_player_device = audio_player_device{}
    @editable
    Audio_FameGranted : audio_player_device = audio_player_device{}
    @editable
    Audio_LessFameGranted : audio_player_device = audio_player_device{}
    @editable
    Audio_ValveChanged : audio_player_device = audio_player_device{}
    @editable
    Audio_HighValve : audio_player_device = audio_player_device{}

    #Actions
    @editable
    Full_AudioPlayer : audio_player_device = audio_player_device{}
    @editable
    Audio_WeaponCritical : audio_player_device = audio_player_device{}
    @editable
    Audio_PickCritical : audio_player_device = audio_player_device{}
    @editable
    Audio_HBWeakExposed : audio_player_device = audio_player_device{}
    @editable
    Audio_HBDeath : audio_player_device = audio_player_device{}
    @editable
    Audio_HBCritical : audio_player_device = audio_player_device{}

    @editable
    SuperStrike_AudioPlayer : audio_player_device = audio_player_device{}
    @editable
    SuperStrikeLoaded_AudioPlayer : audio_player_device = audio_player_device{}

    @editable
    ArrowGranted_AudioPlayer : audio_player_device = audio_player_device{}

    @editable
    Critical_AudioPlayer : audio_player_device = audio_player_device{}

    @editable
    StartChallenge_AudioPlayer : audio_player_device = audio_player_device{}
    @editable
    Damage_AudioPlayer : audio_player_device = audio_player_device{}
    @editable
    HexaPetMove_AudioPlayer : audio_player_device = audio_player_device{}
    @editable
    HexaPetRanged_AudioPlayer : audio_player_device = audio_player_device{}
    @editable
    HexaPetRangedHeal_AudioPlayer : audio_player_device = audio_player_device{}

    @editable
    Heal_AudioPlayer : audio_player_device = audio_player_device{}
    @editable
    Shield_AudioPlayer : audio_player_device = audio_player_device{}


    @editable
    PlayerPays_AudioPlayer : audio_player_device = audio_player_device{}
    @editable
    TimeIsEnding_AudioPlayer : audio_player_device = audio_player_device{}
    @editable
    Audio_Message : audio_player_device = audio_player_device{}
    @editable
    SmallObjetive_AudioPlayer : audio_player_device = audio_player_device{}
    @editable
    PetCreatedObjetive_AudioPlayer : audio_player_device = audio_player_device{}
    @editable
    BigObjetive_AudioPlayer : audio_player_device = audio_player_device{}
    @editable
    ErrorSound_AudioPlayer : audio_player_device = audio_player_device{}
    @editable
    ChestSound_AudioPlayer : audio_player_device = audio_player_device{}
    @editable
    WinSound_AudioPlayer : audio_player_device = audio_player_device{}
    @editable
    LoseSound_AudioPlayer : audio_player_device = audio_player_device{}
    @editable
    WarningSound_AudioPlayer : audio_player_device = audio_player_device{}

    @editable
    AllHardBread_CriticalEffects : []creative_prop_asset = array{DefaultCreativePropAsset}
    @editable
    AllHardBread_DeathEffect : []creative_prop_asset = array{DefaultCreativePropAsset}

    @editable
    ZeroPosition : creative_prop = creative_prop{}

    #Audio Effects
    @editable
    AudioToSingle_PlayerFail : audio_player_device = audio_player_device{}
    @editable
    AudioToSingle_Error : audio_player_device = audio_player_device{}
    @editable
    Audio_SmallObjetive : audio_player_device = audio_player_device{}
    @editable
    Audio_BigObjetive : audio_player_device = audio_player_device{}

    @editable
    AllAudio_Phase0_InestableHB : []audio_player_device = array{}
    var AllPhase0_InestableHB_Used : []logic = array{}

    @editable
    AllAudio_Phase1_InestableHB : []audio_player_device = array{}
    var AllPhase1_InestableHB_Used : []logic = array{}

    @editable
    AllAudio_Phase2_InestableHB : []audio_player_device = array{}
    var AllPhase2_InestableHB_Used : []logic = array{}

    @editable
    AllAudio_Warning_InestableHB : []audio_player_device = array{}
    var AllWarning_InestableHB_Used : []logic = array{}

    @editable
    AllAudio_DroneTakeOff : []audio_player_device = array{}
    var AllDroneTakeOff_Used : []logic = array{}

    @editable
    AllAudio_DroneLanding : []audio_player_device = array{}
    var AllDroneLanding_Used : []logic = array{}


    #Props to Bake
    @editable
    B_Prop : creative_prop_asset = DefaultCreativePropAsset
    @editable
    CB_Prop : creative_prop_asset = DefaultCreativePropAsset
    @editable
    PCS_Prop : creative_prop_asset = DefaultCreativePropAsset

    var AllIsUsedBProps : []logic = array{}
    var AllIsUsedSBProps : []logic = array{}
    var AllIsUsedCBProps : []logic = array{}
    var AllIsUsedPCSProps : []logic = array{}

    #Props to Ship

    
    var BakedBProps : []creative_prop = array{}
    var AllBakedSBProps : []creative_prop = array{}
    var AllBakedCBProps : []creative_prop = array{}
    var AllBakedPCSProps : []creative_prop = array{}

    var AllIsUsedBakedBProps : []logic = array{}
    var AllIsUsedBakedSBProps : []logic = array{}
    var AllIsUsedBakedCBProps : []logic = array{}
    var AllIsUsedBakedPCSProps : []logic = array{}

    #Materials
        #Bread
    AllMat_BakedB : []material = array{Materials.Mat_BakedB_0,Materials.Mat_BakedB_1,Materials.Mat_BakedB_2}
        #SlicedBread
    AllMat_BakedSB : []material = array{Materials.Mat_BakedSB_0,Materials.Mat_BakedSB_1,Materials.Mat_BakedSB_2}
        #CheeseBread
    AllMat_BakedCB : []material = array{Materials.Mat_BakedCB_0,Materials.Mat_BakedCB_1,Materials.Mat_BakedCB_2}
        #PCS
    AllMat_BakedPCS : []material = array{Materials.Mat_BakedPCS, Materials.Mat_BakedPCS, Materials.Mat_BakedPCS}

    #Mastery
    
    @editable
    AllMasteryCameras : []gameplay_camera_fixed_point_device = array{}
    @editable
    Audio_Truck : audio_player_device = audio_player_device{}


    TrySteveBake(GameTime : float) : void =
        if(LastSteveBake < GameTime):
            set LastSteveBake = GameTime + 5.0
            for(Product:=BB_TYPE..PCSB_TYPE, ProductBaker := AllSteveBakers[Product]):
                if(Product=BB_TYPE):
                    spawn:
                        AnimateBake(ProductBaker.GetTransform(), 5.0, false, B_TYPE)
                else if(Product=CBB_TYPE):
                    spawn:
                        AnimateBake(ProductBaker.GetTransform(), 5.0, false, CB_TYPE)
                else if(Product=PCSB_TYPE):
                    spawn:
                        AnimateBake(ProductBaker.GetTransform(), 5.0, false, PCS_TYPE)

    SpawnHealEffect(ShieldMode : logic, Agent : agent)<suspends> : void =
        if(ShieldMode?):
            Heal_AudioPlayer.Play(Agent)
        else:
            Shield_AudioPlayer.Play(Agent)
        if(Character := Agent.GetFortCharacter[]):
            ResultTuple:=SpawnProp(HealAndShieldEffect,Character.GetTransform())
                if(ValidProp := ResultTuple(0)?):
                    Sleep(3.0)
                    ValidProp.Dispose()

    SpawnTeleportEffect(Agent : agent, AgentIndex : int,Destiny : transform) : void =
        if(TeleporterDevice:=AllTeleporterDevices[AgentIndex]):
            FinalTranslation:=vector3{X:=Destiny.Translation.X,Y:=Destiny.Translation.Y,Z:=Destiny.Translation.Z+32.0}
            FinalTransForm:=transform{Translation:=FinalTranslation, Rotation:=Destiny.Rotation, Scale:=NeutralScale}
            if(TeleporterDevice.TeleportTo[FinalTransForm]):
                TeleporterDevice.Teleport(Agent)

    SpawnQuestTracker(Agent : agent, Destiny : transform, AgentIndex : int, TotalGameTime : float) : void =
        ##Print("Shooting projectile")
        if(Character:=Agent.GetFortCharacter[], Character.IsActive[], LastTranslation := AllAgentsLastTranslation[AgentIndex], LastTick := AllAgentsLastTrackerTick[AgentIndex]):

            DistanceFromLastPosition:=Distance(LastTranslation,Character.GetTransform().Translation)

            if(DistanceFromLastPosition>30.0 or (TotalGameTime - LastTick) > 4.0):
                if(set AllAgentsLastTrackerTick[AgentIndex] = TotalGameTime):
                if(set AllAgentsLastTranslation[AgentIndex] = Character.GetTransform().Translation):
                TotalDistance := Distance(Character.GetTransform().Translation,Destiny.Translation)
                Rotation:=GetForwardRotation_Vector3(Character.GetTransform().Translation,Destiny.Translation)

                ObjetiveTraslation := Destiny.Translation
                SpawnTransform:=transform{Translation := Character.GetTransform().Translation, Rotation := Rotation, Scale := Character.GetTransform().Scale}

                TupleSpawn:=SpawnProp(QuestTrackerEffect,SpawnTransform)
                MaybeCreativeProp:=TupleSpawn(0)
                if(ValidProp := MaybeCreativeProp?, ValidProp.IsValid[]):
                    spawn:
                        DespawnProp(ValidProp,5.0)
        else:
            PrintFatalError("On SpawnQuestTracker")

    PlayMasteryCeremony(Agent : agent)<suspends> : void =
        #Audio_BigObjetive.Play(Agent)
        Accolade_Mastery.Award(Agent)
        Sleep(0.3)
        Audio_Truck.Play(Agent)
        if:
            Camera0:=AllMasteryCameras[0]
            Camera1:=AllMasteryCameras[1]
            Camera2:=AllMasteryCameras[2]
        then:
            Camera0.AddTo(Agent)
            Sleep(2.0)
            Camera1.AddTo(Agent)
            Sleep(2.0)
            Camera2.AddTo(Agent)
            Sleep(2.0)
            Camera0.RemoveFrom(Agent)
            Camera1.RemoveFrom(Agent)
            Camera2.RemoveFrom(Agent)
            Audio_Truck.Stop(Agent)
        else:
            PrintFatalError("On DeployIntroGameCeremony, Camera not found!")


    #InitTags() : void =
        <#
        #Bread
        MaybeAllBreadProps : []creative_object_interface := GetCreativeObjectsWithTag(B_prop{})
            #UnbakedBread
        for(Index:=0..Floor[(MaybeAllBreadProps.Length)*1.0/2.0], MaybeBreadProp := MaybeAllBreadProps[Index],Prop := creative_prop[MaybeBreadProp]):
                set AllBProps = AllBProps + array{Prop}
            #BakedBread
        for(Index:=Ceil[(MaybeAllBreadProps.Length)*1.0/2.0]..MaybeAllBreadProps.Length-1, MaybeBreadProp := MaybeAllBreadProps[Index],Prop := creative_prop[MaybeBreadProp]):
            set AllBakedBProps = AllBakedBProps + array{Prop}

        MaybeAllCheeseBreadProps : []creative_object_interface := GetCreativeObjectsWithTag(CB_prop{})
        #CheeseBread
            #UnbakedCheeseBread
        for(Index:=0..Floor[(MaybeAllCheeseBreadProps.Length)*1.0/2.0], MaybeCheeseBreadProp := MaybeAllCheeseBreadProps[Index],Prop := creative_prop[MaybeCheeseBreadProp]):
            set AllCBProps = AllCBProps + array{Prop}
            #BakedCheeseBread
        for(Index:=Ceil[(MaybeAllCheeseBreadProps.Length)*1.0/2.0]..MaybeAllCheeseBreadProps.Length-1, MaybeCheeseBreadProp := MaybeAllCheeseBreadProps[Index], Prop := creative_prop[MaybeCheeseBreadProp]):
            set AllBakedCBProps = AllBakedCBProps + array{Prop}
        #SlicedBread
        MaybeAllSlicedBreadProps : []creative_object_interface := GetCreativeObjectsWithTag(SB_prop{})
            #UnbakedSlicedBread
        for(Index:=0..Floor[(MaybeAllSlicedBreadProps.Length)*1.0/2.0], MaybeSlicedBreadProp := MaybeAllSlicedBreadProps[Index],Prop := creative_prop[MaybeSlicedBreadProp]):
            set AllSBProps = AllSBProps + array{Prop}
            #BakedSlicedBread
        for(Index:=Ceil[(MaybeAllSlicedBreadProps.Length)*1.0/2.0]..MaybeAllSlicedBreadProps.Length-1, MaybeSlicedBreadProp := MaybeAllSlicedBreadProps[Index], Prop := creative_prop[MaybeSlicedBreadProp]):
            set AllBakedSBProps = AllBakedSBProps + array{Prop}
        #PimentoCheeseSandwich
        MaybeAllPCSProps : []creative_object_interface := GetCreativeObjectsWithTag(PCS_prop{})
        for(MaybePCSProp : MaybeAllPCSProps, Prop := creative_prop[MaybePCSProp]):
            set AllPCSProps = AllPCSProps + array{Prop}

        Log("AllBakedBProps: {AllBakedBProps.Length}", LOG_LEVEL_BAKERS)#>


    AnimateDroneMovement<public>(TravelMode : int, DroneProp : creative_prop, InitialTransform : transform, MaybeEffectSound : ?audio_player_device)<suspends> : void =

        Log("Animating drone shipment",LOG_LEVEL_1DETAILING)
        
        TAKEOFFMODE : int = 0
        RETURNMODE : int = 1

        FinalPosition := vector3{X:=InitialTransform.Translation.X,Y:=InitialTransform.Translation.Y+4096.0,Z:=InitialTransform.Translation.Z}
        FinalTransform := transform{Translation:=FinalPosition,Rotation:=rotation{},Scale:=vector3{X:=1.0,Y:=1.0,Z:=1.0}}

        var RotationApplied0 : rotation = rotation{}
        var RotationApplied1 : rotation = rotation{}
        var AppliedVector0 : vector3 = vector3{X:=0.0,Y:=0.0,Z:=1024.0}
        var AppliedVector1 : vector3 = vector3{X:=0.0,Y:=0.0,Z:=0.0}
        var AppliedVector2 : vector3 = vector3{X:=0.0,Y:=0.0,Z:=0.0}
        var AppliedVector3 : vector3 = vector3{X:=0.0,Y:=0.0,Z:=0.0}

        Frame0Time := 1.0
        Frame1Time := 0.25
        Frame2Time := 2.0
        Frame3Time := 2.0
        LandingDelaySound := Frame0Time + Frame1Time + Frame2Time + Frame3Time - 1.25

        if(TravelMode=TAKEOFFMODE):
            if(DroneProp.TeleportTo[InitialTransform]):
                spawn:
                    PlaySound(DRONETAKEOFF_SOUND,0.1,false,option{InitialTransform})
                set RotationApplied0 = MakeRotationFromYawPitchRollDegrees(DRONE_Send_RotationYaw, DRONE_Send_RotationPitch,DRONE_Send_RotationRoll)
                set RotationApplied1 = MakeRotationFromYawPitchRollDegrees(DRONE_Return_RotationYaw, DRONE_Return_RotationPitch,DRONE_Return_RotationRoll)
                set AppliedVector1 = vector3{X:=0.0,Y:=100.0,Z:=0.0}
                set AppliedVector2 = vector3{X:=0.0,Y:=3996.0,Z:=0.0}
                set AppliedVector3 = vector3{X:=0.0,Y:=0.0,Z:=-1024.0}
        else if(TravelMode=RETURNMODE):
            if(DroneProp.TeleportTo[FinalTransform]):
                spawn:
                    PlaySound(DRONELANDING_SOUND,LandingDelaySound,false,option{InitialTransform})
                set RotationApplied0 = MakeRotationFromYawPitchRollDegrees(DRONE_Return_RotationYaw, DRONE_Return_RotationPitch,DRONE_Return_RotationRoll)
                set RotationApplied1 = MakeRotationFromYawPitchRollDegrees(DRONE_Send_RotationYaw, DRONE_Send_RotationPitch,DRONE_Send_RotationRoll)
                set AppliedVector1 = vector3{X:=0.0,Y:=-100.0,Z:=0.0}
                set AppliedVector2 = vector3{X:=0.0,Y:=-3996.0,Z:=0.0}
                set AppliedVector3 = vector3{X:=0.0,Y:=0.0,Z:=-1024.0}
        else:
            #Print("Fatal error on AnimateDroneMovement. TravelMode: {TravelMode} not found")

        #NeutralScale := vector3{X:=1.0,Y:=1.0,Z:=1.0}

        MovementKeyFrame0 : keyframe_delta = keyframe_delta #Key Frame 1 to begining
        {
            DeltaLocation := AppliedVector0
            DeltaRotation := rotation{}
            DeltaScale := NeutralScale
            Time := Frame0Time
            Interpolation := EaseInOut
        }

        MovementKeyFrame1 : keyframe_delta = keyframe_delta #Key Frame 1 to begining
        {
            DeltaLocation := AppliedVector1
            DeltaRotation := RotationApplied0
            DeltaScale := NeutralScale
            Time := 0.25
            Interpolation := EaseInOut
        }

        MovementKeyFrame2 : keyframe_delta = keyframe_delta #Key Frame 1 to begining
        {
            DeltaLocation := AppliedVector2
            DeltaRotation := rotation{}
            DeltaScale := NeutralScale
            Time := 2.0
            Interpolation := EaseInOut
        }

        MovementKeyFrame3 : keyframe_delta = keyframe_delta #Key Frame 1 to begining
        {
            DeltaLocation := AppliedVector3
            DeltaRotation := RotationApplied1
            DeltaScale := NeutralScale
            Time := 2.0
            Interpolation := EaseInOut
        }
        
        if(PropController := DroneProp.GetAnimationController[]):
            # Make an array of key frames
            FortVaderKeyFrames :  []keyframe_delta = array {MovementKeyFrame0, MovementKeyFrame1, MovementKeyFrame2, MovementKeyFrame3}
            # Set the animation to play
            PropController.SetAnimation(FortVaderKeyFrames, ?Mode:=animation_mode.OneShot)
            # Play Animation
            Sleep(0.5)
            PropController.Play()
            if(ValidEffectSound := MaybeEffectSound?):
                ValidEffectSound.Play()
            loop:
                State := PropController.GetState()
                if(State = animation_controller_state.Stopped):
                    if(TravelMode=0):
                        if(DroneProp.TeleportTo[FinalTransform]):
                    else if(TravelMode=1):
                        if(DroneProp.TeleportTo[InitialTransform]):
                    break
                Sleep(0.04)
        

    # Runs when the device is started in a running game
    OnBegin<override>()<suspends>:void=
        #Audios
        for(AudioPlayer : AllAudio_DroneTakeOff):
            set AllDroneTakeOff_Used = AllDroneTakeOff_Used + array{false}
        for(AudioPlayer : AllAudio_DroneLanding):
            set AllDroneLanding_Used = AllDroneLanding_Used + array{false}
        for(AudioPlayer : AllAudio_Phase0_InestableHB):
            set AllPhase0_InestableHB_Used = AllPhase0_InestableHB_Used + array{false}
        for(AudioPlayer : AllAudio_Phase1_InestableHB):
            set AllPhase1_InestableHB_Used = AllPhase1_InestableHB_Used + array{false}
        for(AudioPlayer : AllAudio_Phase2_InestableHB):
            set AllPhase2_InestableHB_Used = AllPhase2_InestableHB_Used + array{false}
        for(AudioPlayer : AllAudio_Warning_InestableHB):
            set AllWarning_InestableHB_Used = AllWarning_InestableHB_Used + array{false}

        #InitTags
        #InitTags()
            #Product props
            <#
        for(Index :=0..AllBProps.Length-1):
            set AllIsUsedBProps = AllIsUsedBProps + array{false}

        for(Index :=0..AllCBProps.Length-1):
            set AllIsUsedCBProps = AllIsUsedCBProps + array{false}

        for(Index :=0..AllPCSProps.Length-1):
            set AllIsUsedPCSProps = AllIsUsedPCSProps + array{false}

        for(Index :=0..AllSBProps.Length-1):
            set AllIsUsedSBProps = AllIsUsedSBProps + array{false}#>

            #ShipProducts
            <#
        for(Index :=0..AllBakedBProps.Length-1, Prop := AllBakedBProps[Index]):
            Prop.SetMaterial(Materials.Mat_BakedB_2, ?Index:=0)
            set AllIsUsedBakedBProps = AllIsUsedBakedBProps + array{false}

        for(Index :=0..AllBakedCBProps.Length-1, Prop := AllBakedCBProps[Index]):
            Prop.SetMaterial(Materials.Mat_BakedSB_2, ?Index:=0)
            set AllIsUsedBakedCBProps = AllIsUsedBakedCBProps + array{false}

        for(Index :=0..AllBakedCBProps.Length-1, Prop := AllBakedCBProps[Index]):
            Prop.SetMaterial(Materials.Mat_BakedCB_2, ?Index:=0)
            set AllIsUsedBakedCBProps = AllIsUsedBakedCBProps + array{false} 

        for(Index :=0..AllBakedPCSProps.Length-1):
            set AllIsUsedBakedPCSProps = AllIsUsedBakedPCSProps + array{false}#>

    StopAllMusics(Agent : agent) : void =
        HexaGameMusic_AudioPlayer.Stop(Agent)
        DungeonMusic_AudioPlayer.Stop(Agent)
        DungeonCombatMusic_AudioPlayer.Stop(Agent)
        BakersMusic_AudioPlayer.Stop(Agent)
        DesertMusic_AudioPlayer.Stop(Agent)
        SecretMusic_AudioPlayer.Stop(Agent)

    GetRangeAttackProp() : creative_prop_asset =
        return RangeAttackEffect

    GetMeleeAttackProp() : creative_prop_asset =
        return MeleeAttackEffect

    GetDefendProp() : creative_prop_asset =
        return DefendEffect

    PlaySound(SoundType : int, Delay : float, MaybeAgent : ?agent, MaybeTransform : ?transform, ?StringMessage : string = "")<suspends> : void =
        var AudioPlayer_ToUse : audio_player_device = audio_player_device{}
        var IndexAudioToUse : int = -1
            #Drone Take off
        if(SoundType=DRONETAKEOFF_SOUND):
            loop:
                if(IsUsed:=AllDroneTakeOff_Used[IndexAudioToUse], not IsUsed?, set AllDroneTakeOff_Used[IndexAudioToUse] = true):
                    if(AudioPlayer:=AllAudio_DroneTakeOff[IndexAudioToUse]):
                        set AudioPlayer_ToUse=AudioPlayer
                    break
                if(IndexAudioToUse>AllAudio_DroneTakeOff.Length-1):
                    PrintWarning("Audio player not found")
                    set IndexAudioToUse = -1
                    break
                set IndexAudioToUse = IndexAudioToUse + 1
        else if(SoundType=DRONELANDING_SOUND):
            loop:
                if(IsUsed:=AllDroneLanding_Used[IndexAudioToUse], not IsUsed?, set AllDroneLanding_Used[IndexAudioToUse] = true):
                    if(AudioPlayer:=AllAudio_DroneLanding[IndexAudioToUse]):
                        set AudioPlayer_ToUse=AudioPlayer
                    break
                if(IndexAudioToUse>AllAudio_DroneLanding.Length-1):
                    PrintWarning("Audio player not found")
                    set IndexAudioToUse = -1
                    break
                set IndexAudioToUse = IndexAudioToUse + 1

        else if(SoundType=IHB_PHASE0_SOUND):
            loop:
                if(IsUsed:=AllPhase0_InestableHB_Used[IndexAudioToUse], not IsUsed?, set AllPhase0_InestableHB_Used[IndexAudioToUse] = true):
                    if(AudioPlayer:=AllAudio_Phase0_InestableHB[IndexAudioToUse]):
                        set AudioPlayer_ToUse=AudioPlayer
                    break
                if(IndexAudioToUse>AllAudio_Phase0_InestableHB.Length-1):
                    PrintWarning("Audio player not found")
                    set IndexAudioToUse = -1
                    break
                set IndexAudioToUse = IndexAudioToUse + 1
        else if(SoundType=IHB_PHASE1_SOUND):
            loop:
                if(IsUsed:=AllPhase1_InestableHB_Used[IndexAudioToUse], not IsUsed?, set AllPhase1_InestableHB_Used[IndexAudioToUse] = true):
                    if(AudioPlayer:=AllAudio_Phase1_InestableHB[IndexAudioToUse]):
                        set AudioPlayer_ToUse=AudioPlayer
                    break
                if(IndexAudioToUse>AllAudio_Phase1_InestableHB.Length-1):
                    PrintWarning("Audio player not found")
                    set IndexAudioToUse = -1
                    break
                set IndexAudioToUse = IndexAudioToUse + 1
        else if(SoundType=IHB_PHASE2_SOUND):
            loop:
                if(IsUsed:=AllPhase2_InestableHB_Used[IndexAudioToUse], not IsUsed?, set AllPhase2_InestableHB_Used[IndexAudioToUse] = true):
                    if(AudioPlayer:=AllAudio_Phase2_InestableHB[IndexAudioToUse]):
                        set AudioPlayer_ToUse=AudioPlayer
                    break
                if(IndexAudioToUse>AllAudio_Phase2_InestableHB.Length-1):
                    PrintWarning("Audio player not found")
                    set IndexAudioToUse = -1
                    break
                set IndexAudioToUse = IndexAudioToUse + 1
        else if(SoundType=IHB_WARNING_SOUND):
            loop:
                if(IsUsed:=AllWarning_InestableHB_Used[IndexAudioToUse], not IsUsed?, set AllWarning_InestableHB_Used[IndexAudioToUse] = true):
                    if(AudioPlayer:=AllAudio_Warning_InestableHB[IndexAudioToUse]):
                        set AudioPlayer_ToUse=AudioPlayer
                    break
                if(IndexAudioToUse>AllAudio_Warning_InestableHB.Length-1):
                    PrintWarning("Audio player not found")
                    set IndexAudioToUse = -1
                    break
                set IndexAudioToUse = IndexAudioToUse + 1
        else if(SoundType=SMALLOBJETIVE_SOUND, Agent := MaybeAgent?):
            Audio_SmallObjetive.Play(Agent)
        else if(SoundType=BIGOBJETIVE_SOUND, Agent := MaybeAgent?):
            Audio_BigObjetive.Play(Agent)
        else if(SoundType=WEAPONCRITICAL_SOUND, Agent := MaybeAgent?):
            Audio_WeaponCritical.Play(Agent)
        else if(SoundType=PICKCRITICAL_SOUND, Agent := MaybeAgent?):
            Audio_PickCritical.Play(Agent)
        else if(SoundType=HIGHVALVE_SOUND):
            set AudioPlayer_ToUse = Audio_HighValve
        else if(SoundType=DAMAGE_SOUND, Agent := MaybeAgent?):
            Damage_AudioPlayer.Play(Agent)
        else if(SoundType=HEAL_SOUND, Agent := MaybeAgent?):
            Heal_AudioPlayer.Play(Agent)
        else if(SoundType=SHIELD_SOUND, Agent := MaybeAgent?):
            Shield_AudioPlayer.Play(Agent)
        else if(SoundType=TIMEENDING_SOUND, Agent := MaybeAgent?):
            TimeIsEnding_AudioPlayer.Play(Agent)
        else if(SoundType=PETCREATED_SOUND, Agent := MaybeAgent?):
            PetCreatedObjetive_AudioPlayer.Play(Agent)
        else if(SoundType=ERROR_SOUND, Agent := MaybeAgent?):
            ErrorSound_AudioPlayer.Play(Agent)
        else if(SoundType=CHEST_SOUND, Agent := MaybeAgent?):
            ChestSound_AudioPlayer.Play(Agent)
        else if(SoundType=WIN_SOUND, Agent := MaybeAgent?):
            WinSound_AudioPlayer.Play(Agent)
        else if(SoundType=LOSE_SOUND, Agent := MaybeAgent?):
            LoseSound_AudioPlayer.Play(Agent)
        else if(SoundType=SUPERSTRIKE_SOUND, Agent := MaybeAgent?):
            SuperStrike_AudioPlayer.Play(Agent)
        else if(SoundType=DUNGEONMUSIC_SOUND, Agent := MaybeAgent?):
            DungeonMusic_AudioPlayer.Play(Agent)
        else if(SoundType=DUNGEONMINIBOSSMUSIC_SOUND, Agent := MaybeAgent?):
            DungeonCombatMusic_AudioPlayer.Play(Agent)
        else if(SoundType=STARTCHALLENGE_SOUND, Agent := MaybeAgent?):
            StartChallenge_AudioPlayer.Play(Agent)
        else if(SoundType=HEXAGAMEMUSIC_SOUND, Agent := MaybeAgent?):
            HexaGameMusic_AudioPlayer.Play(Agent)
        else if(SoundType=BAKERSMUSIC_SOUND, Agent := MaybeAgent?):
            BakersMusic_AudioPlayer.Play(Agent)
        else if(SoundType=DESERTMUSIC_SOUND, Agent := MaybeAgent?):
            DesertMusic_AudioPlayer.Play(Agent)
        else if(SoundType=SECRETMUSIC_SOUND, Agent := MaybeAgent?):
            SecretMusic_AudioPlayer.Play(Agent)
        else if(SoundType=FULL_SOUND, Agent := MaybeAgent?):
            Full_AudioPlayer.Play(Agent)
            #Print("TO DO: Visualice `{StringMessage}` on VFX.PlayErrorAudio")
        else if(SoundType=MESSAGE_SOUND, Agent := MaybeAgent?):
            Audio_Message.Play(Agent)
        else if(SoundType=WARNING_SOUND, Agent := MaybeAgent?):
            WarningSound_AudioPlayer.Play(Agent)

        if(PositionTransform:=MaybeTransform?):
            #Teleport Audio Device
            if(AudioPlayer_ToUse.TeleportTo[PositionTransform]):
                Sleep(0.1)
                Sleep(Delay)
                AudioPlayer_ToUse.Play()
                Sleep(3.0)
        #Clean
        if(IndexAudioToUse<>-1):
            if(SoundType=DRONETAKEOFF_SOUND, set AllDroneTakeOff_Used[IndexAudioToUse]=false):
            else if(SoundType=DRONELANDING_SOUND, set AllDroneLanding_Used[IndexAudioToUse]=false):
            else if(SoundType=IHB_PHASE0_SOUND, set AllPhase0_InestableHB_Used[IndexAudioToUse]=false):
            else if(SoundType=IHB_PHASE1_SOUND, set AllPhase1_InestableHB_Used[IndexAudioToUse]=false):
            else if(SoundType=IHB_PHASE2_SOUND, set AllPhase2_InestableHB_Used[IndexAudioToUse]=false):
            else if(SoundType=IHB_WARNING_SOUND, set AllWarning_InestableHB_Used[IndexAudioToUse]=false):

    StopSound(SoundType : int, Delay : float, MaybeAgent : ?agent, MaybeTransform : ?transform, ?StringMessage : string = "")<suspends> : void =
        var AudioPlayer_ToUse : audio_player_device = audio_player_device{}
        var IndexAudioToUse : int = -1
            #Drone Take off
        Sleep(0.0)
        if(SoundType=DUNGEONMUSIC_SOUND, Agent := MaybeAgent?):
            DungeonMusic_AudioPlayer.Stop(Agent)
        else if(SoundType=DUNGEONMINIBOSSMUSIC_SOUND, Agent := MaybeAgent?):
            DungeonCombatMusic_AudioPlayer.Stop(Agent)
        else if(SoundType=HEXAGAMEMUSIC_SOUND, Agent := MaybeAgent?):
            HexaGameMusic_AudioPlayer.Stop(Agent)
        else if(SoundType=BAKERSMUSIC_SOUND, Agent := MaybeAgent?):
            BakersMusic_AudioPlayer.Stop(Agent)
        else if(SoundType=DESERTMUSIC_SOUND, Agent := MaybeAgent?):
            DesertMusic_AudioPlayer.Stop(Agent)
        else if(SoundType=SECRETMUSIC_SOUND, Agent := MaybeAgent?):
            SecretMusic_AudioPlayer.Stop(Agent)
            

    RunPetCreation(Agent : agent, PetPropAsset : creative_prop_asset, Tier : int)<suspends> : void =
        PetBlenderPropPosition:=PetBlenderProp.GetTransform().Translation
        SpawnPosition:=PetBlenderPropPosition+vector3{X:=80.0,Y:=-28.0,Z:=228.0}
        SpawnRotation:=MakeRotationFromYawPitchRollDegrees(0.0,0.0,0.0)
        #SpawnRotation:=MakeRotationFromYawPitchRollDegrees(180.0,0.0,0.0)
        #SpawnRotation:=PetBlenderProp.GetTransform().Rotation.ApplyYaw(180.0) 
        var NormalScale : vector3 = vector3{}

        if(Tier=1):
            set NormalScale = PetScale_Tier1
            #set PivotVariation = PetPivotVariation_Tier1
        if(Tier=2):
            set NormalScale = PetScale_Tier2
            #set PivotVariation = PetPivotVariation_Tier2
        if(Tier=3):
            set NormalScale = PetScale_Tier3
            #set PivotVariation = PetPivotVariation_Tier3

        PetBlenderPropTransform:=transform{Scale:=NormalScale,Rotation:=SpawnRotation,Translation:=SpawnPosition}
        ResultTuple:=SpawnProp(PetPropAsset,PetBlenderPropTransform)
        #Print("Spawning on {SpawnPosition.X} {SpawnPosition.Y} {SpawnPosition.Z}")
        if(PetProp := ResultTuple(0)?):
            SetStar(PetProp,Tier)
            RemoveHealthUI(PetProp)
            #Run pet creation
            Camera_PetCreation.AddTo(Agent)

            Frame0_VectorApplied:=vector3{X:=8.0,Y:=0.0,Z:=-64.0}
            Frame0_RotationApplied:=MakeRotationFromYawPitchRollDegrees(0.0, 90.0, 0.0)
            Frame1_VectorApplied:=vector3{X:=64.0,Y:=0.0,Z:=-64.0}
            Frame1_RotationApplied:=MakeRotationFromYawPitchRollDegrees(0.0, 270.0, 0.0)

            #if(PetProp.TeleportTo[PetBlenderPropPosition,SpawnRotation]):

            MovementKeyFrame0 : keyframe_delta = keyframe_delta #Key Frame 0
            {
                DeltaLocation := Frame0_VectorApplied
                #DeltaRotation := rotation{}
                DeltaRotation := Frame0_RotationApplied
                DeltaScale := NeutralScale
                Time := 0.5
                Interpolation := EaseIn
            }

            MovementKeyFrame1 : keyframe_delta = keyframe_delta #Key Frame 1
            {
                DeltaLocation := Frame1_VectorApplied
                #DeltaRotation := rotation{}
                DeltaRotation := Frame1_RotationApplied
                DeltaScale := NeutralScale
                Time := 1.0
                Interpolation := EaseIn
            }

            MovementKeyFrame2 : keyframe_delta = keyframe_delta #Key Frame 1
            {
                DeltaLocation := vector3{}
                DeltaRotation := rotation{}
                DeltaScale := NeutralScale
                Time := 1.0
                Interpolation := EaseIn
            }

            if(PropController := PetProp.GetAnimationController[]):
                # Make an array of key frames
                FortVaderKeyFrames :  []keyframe_delta = array { MovementKeyFrame0 , MovementKeyFrame1, MovementKeyFrame2}
                # Set the animation to play
                PropController.SetAnimation(FortVaderKeyFrames, ?Mode:=animation_mode.OneShot)
                PropController.Play()
                loop:
                    State := PropController.GetState()
                    if(State = animation_controller_state.Stopped):
                        break
                    Sleep(0.04)
            PetProp.Dispose()
            Camera_PetCreation.RemoveFrom(Agent)
        else:
            PrintFatalError("On RunPetCreation. Spawn result not valid")
        

    Change_Material(Prop : creative_prop, FrameIndex : int, ProductType : int) : void =
        if(ProductType=B_TYPE):
            if(MaterialSelected := AllMat_BakedB[FrameIndex-1]):
                Prop.SetMaterial(MaterialSelected,?Index:=0)
        else if(ProductType=CB_TYPE):
            if(MaterialSelected := AllMat_BakedCB[FrameIndex-1]):
                Prop.SetMaterial(MaterialSelected,?Index:=0)
        else if(ProductType=PCS_TYPE):
            if(MaterialSelected := AllMat_BakedSB[FrameIndex-1]):
                Prop.SetMaterial(MaterialSelected,?Index:=0)
        else:
            PrintFatalError("Fatal error on Change_Material. ProductType: {ProductType} invalid")



    Spawn_HardBreadDeath<public>(Agent : agent, HardBreadProp : creative_prop, HBType : int, ?ZVariation : float = 0.0)<suspends> : void =
        if(DeathEffect := AllHardBread_DeathEffect[HBType]): 
            Translation:=vector3{X:=HardBreadProp.GetTransform().Translation.X,Y:=HardBreadProp.GetTransform().Translation.Y,Z:=HardBreadProp.GetTransform().Translation.Z+128.0+ZVariation}
            Audio_HBDeath.Play(Agent)
            Sleep(0.1)
            SpawnTuple:=SpawnProp(DeathEffect,transform{Translation:=Translation,Rotation:=rotation{},Scale:=vector3{X:=1.0,Y:=1.0,Z:=1.0}})
            MaybeProp:=SpawnTuple(0)
            Result := SpawnTuple(1)
            if(Result = spawn_prop_result.Ok):
                Log("Spawn OK", LOG_LEVEL_NIAGGARA)
            else:
                Log("ERROR ON Spawn", LOG_LEVEL_NIAGGARA)
            if(ValidProp := MaybeProp?):
                Log("Spawning death props", LOG_LEVEL_NIAGGARA)
                spawn:
                    DespawnProp(ValidProp,5.0)

    AwardAccolade(AccoladeType : int, Agent : agent) : void =
        if(AccoladeType=PICKAXE_ACCOLADE):
            Accolade_PickaxeImprovement.Award(Agent)
        else if(AccoladeType=BAKER_ACCOLADE):
            Accolade_BakerImprovement.Award(Agent)
        else if(AccoladeType=DRONESEND_ACCOLADE):
            Accolade_DroneSend.Award(Agent)
        else if(AccoladeType=HARDBREAD_ACCOLADE):
            Accolade_HardBreadDestroyed.Award(Agent)
        else if(AccoladeType=INVENTORY_ACCOLADE):
            Accolade_Inventory.Award(Agent)
        else if(AccoladeType=AGEDSAPPHIRES_ACCOLADE):
            Accolade_AgedSapphires.Award(Agent)


    PlayWeakPointExposed(Agent : agent) : void =
        Audio_HBWeakExposed.Play(Agent)

    Spawn_HardBreadCriticalStrike<public>(Agent : agent, HBType : int) : void =
        if(Character:=player[Agent].GetFortCharacter[]):
            if(CriticalEffect := AllHardBread_CriticalEffects[HBType]):

                ViewLocation:=Character.GetViewLocation()
                ViewRotation:=Character.GetViewRotation()
                PointingVector:=ViewRotation.GetLocalForward()*200.0
                NewPosition:=ViewLocation+PointingVector

                Audio_HBCritical.Play(Agent)
                SpawnTuple:=SpawnProp(CriticalEffect,transform{Translation:=NewPosition,Rotation:=rotation{},Scale:=vector3{X:=1.0,Y:=1.0,Z:=1.0}})
                MaybeProp:=SpawnTuple(0)
                Result := SpawnTuple(1)
                if(Result = spawn_prop_result.Ok):
                    Log("Spawn OK", LOG_LEVEL_NIAGGARA)
                else:
                    Log("ERROR ON Spawn", LOG_LEVEL_NIAGGARA)
                if(ValidProp := MaybeProp?):
                    Log("Spawning Critical props", LOG_LEVEL_NIAGGARA)
                    spawn:
                        DespawnProp(ValidProp,5.0)


    DespawnProp(CreativeProp : creative_prop, Time : float)<suspends> : void =
        Sleep(Time)
        if(CreativeProp.IsValid[]):
            CreativeProp.Dispose()


    AnimateBake<public>(BakeTransform : transform, Speed : float, MaybeEffectSound : ?audio_player_device, ProductType : int)<suspends> : void =

        var Index : int = 0
        var Prop : creative_prop = creative_prop{}
        var StartPoint : vector3 = vector3{}
        var VectorToEndPoint : vector3 = vector3{}

        var VectorToMid1Point  : vector3 = vector3{}
        var VectorToMid2Point  : vector3 = vector3{}
        var VectorToMid3Point  : vector3 = vector3{}
        var VectorToMid4Point  : vector3 = vector3{}

        VectorZero := vector3{X:=0.0,Y:=0.0,Z:=0.0}

        Log("Animating Bake product {ProductType}",LOG_LEVEL_BAKERS)

        <#
        case(ProductType):
            0=>
                loop:
                    if(Value := AllIsUsedBProps[Index], not Value?, AuxProp := AllBProps[Index], set Prop = AuxProp):
                        if(set AllIsUsedBProps[Index] = true, UnBakedMaterial := AllMat_BakedB[0]):
                            Prop.SetMaterial(UnBakedMaterial, ?Index:=0)
                            Log("Animating Bake. Bread selected: {Index}", LOG_LEVEL_BAKERS)
                        break
                    else:
                        set Index = Index + 1
                    if(Index > AllBProps.Length):
                        #Print("Fatal error on AnimateBake. Bread to use as animation don´t found")
                        break
            1=>
                loop:
                    if(Value := AllIsUsedCBProps[Index], not Value?, AuxProp := AllCBProps[Index], set Prop = AuxProp):
                        if(set AllIsUsedCBProps[Index] = true, UnBakedMaterial := AllMat_BakedCB[0]):
                            Prop.SetMaterial(UnBakedMaterial, ?Index:=0)
                        break
                    else:
                        set Index = Index + 1
                    if(Index > AllCBProps.Length):
                        #Print("Fatal error on AnimateBake. Cheese Bread to use as animation don´t found")
                        break

            2=>
                loop:
                    if(Value := AllIsUsedPCSProps[Index], not Value?, AuxProp := AllPCSProps[Index], set Prop = AuxProp):
                        if(set AllIsUsedPCSProps[Index] = true):
                        break
                    else:
                        set Index = Index + 1
                    if(Index > AllPCSProps.Length):
                        #Print("Fatal error on AnimateBake. PCS Bread to use as animation don´t found")
                        break
            3=>
                loop:
                    if(Value := AllIsUsedSBProps[Index], not Value?, AuxProp := AllSBProps[Index], set Prop = AuxProp):
                        if(set AllIsUsedSBProps[Index] = true, UnBakedMaterial := AllMat_BakedSB[0]):
                            Prop.SetMaterial(UnBakedMaterial, ?Index:=0)
                        break
                    else:
                        set Index = Index + 1
                    if(Index > AllSBProps.Length):
                        #Print("Fatal error on AnimateBake. Sliced Bread to use as animation don´t found")
                        break
            
            _=> #Print("FATAL ERROR on AnimateBake. Product Type: {ProductType} not valid")#>

        Log("Starting baking animation {Index}", LOG_LEVEL_BAKERS)

        #PropPoint := Prop.GetTransform().Translation

        if(ProductType = B_TYPE or ProductType = CB_TYPE):
            Log("Starting baking animation {Index}", LOG_LEVEL_BAKERS)
            set StartPoint = BakeTransform.Translation + PRODUCT_STARTPOSITION
            set VectorToMid1Point = (BakeTransform.Translation + PRODUCT_MIDPOSITION)-StartPoint
            set VectorToMid2Point = VectorZero
            set VectorToMid3Point = VectorZero
            set VectorToMid4Point = VectorZero
            set VectorToEndPoint = (BakeTransform.Translation + PRODUCT_ENDPOSITION)-(StartPoint+VectorToMid1Point)
        if(ProductType = SB_TYPE):
            set StartPoint = BakeTransform.Translation + SB_STARTPOSITION
            set VectorToMid1Point = (BakeTransform.Translation + SB_MIDPOSITION)-StartPoint
            set VectorToMid2Point = VectorZero
            set VectorToMid3Point = VectorZero
            set VectorToMid4Point = VectorZero
            set VectorToEndPoint = (BakeTransform.Translation + SB_ENDPOSITION)-(StartPoint+VectorToMid1Point)
        if(ProductType = PCS_TYPE):
            set StartPoint = BakeTransform.Translation + PCS_STARTPOSITION
            set VectorToMid1Point = (BakeTransform.Translation + PCS_MID1POSITION)-StartPoint
            set VectorToMid2Point = (BakeTransform.Translation + PCS_MID2POSITION)-(StartPoint+VectorToMid1Point)
            set VectorToMid3Point = (BakeTransform.Translation + PCS_MID3POSITION)-(StartPoint+VectorToMid1Point+VectorToMid2Point)
            set VectorToMid4Point = (BakeTransform.Translation + PCS_MID4POSITION)-(StartPoint+VectorToMid1Point+VectorToMid2Point+VectorToMid3Point)
            set VectorToEndPoint = (BakeTransform.Translation + PCS_ENDPOSITION)-(StartPoint+VectorToMid1Point+VectorToMid2Point+VectorToMid3Point+VectorToMid4Point)
            #set VectorToMidPoint = (BakeTransform.Translation + PCS_MIDPOSITION)-StartPoint
            #set VectorToEndPoint = (BakeTransform.Translation + PCS_ENDPOSITION)-(StartPoint+VectorToMidPoint)
        
        EmptyRotation : rotation = rotation{}
        StandarScale := BakeTransform.Scale

        TicksAnimationFrame1 := 0.5
        TicksAnimationFrame2 := 2.0
        TicksAnimationFrame3 := 3.0
        TicksAnimationFrame4 := 3.0
        TicksAnimationFrame5 := 0.5
        var TotalTicksOfAnimation : float = TicksAnimationFrame1 + TicksAnimationFrame2 + TicksAnimationFrame3 + TicksAnimationFrame4 + TicksAnimationFrame5

        #if(ProductType = SB_TYPE):
        #    set TotalTicksOfAnimation = (TotalTicksOfAnimation * 2.0) - 1.0
        #if(ProductType = PCS_TYPE):
        #    set TotalTicksOfAnimation = TotalTicksOfAnimation * 2.0

        Frame1Time : float = (Speed/TotalTicksOfAnimation)*TicksAnimationFrame1
        Frame2Time : float = (Speed/TotalTicksOfAnimation)*TicksAnimationFrame2
        Frame3Time : float = (Speed/TotalTicksOfAnimation)*TicksAnimationFrame3
        Frame4Time : float = (Speed/TotalTicksOfAnimation)*TicksAnimationFrame4
        Frame5Time : float = (Speed/TotalTicksOfAnimation)*TicksAnimationFrame5

        <#
        if(Prop.TeleportTo[NewTransform:=transform{Translation := StartPoint, Rotation := BakeTransform.Rotation, Scale := StandarScale}]):
        else:
            PrintFatalError("On Animate Bake. Couldn´t teleport")#>

        var PropToSpawn : creative_prop_asset = DefaultCreativePropAsset

        if(ProductType=B_TYPE):
            set PropToSpawn = B_Prop
        else if(ProductType=CB_TYPE):
            set PropToSpawn = CB_Prop
        else if(ProductType=PCS_TYPE):
            set PropToSpawn = PCS_Prop
            
            
        ResultTuple:=SpawnProp(PropToSpawn,NewTransform:=transform{Translation := StartPoint, Rotation := BakeTransform.Rotation, Scale := StandarScale})
        if(ValidProp := ResultTuple(0)?, set Prop = ValidProp):
            spawn:
                DespawnProp(ValidProp,Frame1Time+Frame2Time+Frame3Time+Frame4Time+Frame5Time+5.0)
            if(ProductType=B_TYPE, UnBakedMaterial := AllMat_BakedB[0]):
                Prop.SetMaterial(UnBakedMaterial, ?Index:=0)
            else if(ProductType=CB_TYPE, UnBakedMaterial := AllMat_BakedCB[0]):
                Prop.SetMaterial(UnBakedMaterial, ?Index:=0)
            else if(ProductType=PCS_TYPE, UnBakedMaterial := AllMat_BakedPCS[0]):
                Prop.SetMaterial(UnBakedMaterial, ?Index:=0)


            MovementKeyFrame1 : keyframe_delta = keyframe_delta #Key Frame 1 to begining
            {
                DeltaLocation := VectorToMid1Point
                DeltaScale := StandarScale
                DeltaRotation := EmptyRotation
                Time := Frame1Time
                Interpolation := EaseOut
            }


            MovementKeyFrame2 : keyframe_delta = keyframe_delta #Key Frame 1 to begining
            {
                DeltaLocation := VectorToMid2Point
                DeltaScale := StandarScale
                DeltaRotation := EmptyRotation
                Time := Frame2Time
                Interpolation := EaseIn
                
            }

            MovementKeyFrame3 : keyframe_delta = keyframe_delta #Key Frame 1 to begining
            {
                DeltaLocation := VectorToMid3Point
                DeltaScale := StandarScale
                DeltaRotation := EmptyRotation
                Time := Frame3Time
                Interpolation := EaseIn
            }

            MovementKeyFrame4 : keyframe_delta = keyframe_delta #Key Frame 1 to begining
            {
                DeltaLocation := VectorToMid4Point
                DeltaScale := StandarScale
                DeltaRotation := EmptyRotation
                Time := Frame4Time
                Interpolation := EaseIn
            }

            MovementKeyFrame5 : keyframe_delta = keyframe_delta #Key Frame 1 to begining
            {
                DeltaLocation := VectorToEndPoint
                DeltaScale := StandarScale
                DeltaRotation := EmptyRotation
                Time := Frame5Time
                Interpolation := EaseIn
            }

            if(PropController := Prop.GetAnimationController[]):
                # Make an array of key frames
                FortVaderKeyFrames :  []keyframe_delta = array { MovementKeyFrame1 , MovementKeyFrame2,  MovementKeyFrame3, MovementKeyFrame4, MovementKeyFrame5}
                # Set the animation to play
                PropController.SetAnimation(FortVaderKeyFrames, ?Mode:=animation_mode.OneShot)
                if(ProductType <> PCS_TYPE):
                    PropController.KeyframeReachedEvent.Subscribe(handler_random{MaybeProp := option{Prop}, MaybeVFXManager := option{Self}, PropType := ProductType}.HandlerChange_Material)   
                # Play Animation
                PropController.Play()
                if(ValidEffectSound := MaybeEffectSound?):
                    ValidEffectSound.Play()
                loop:
                    State := PropController.GetState()
                    if(State = animation_controller_state.Stopped):
                        if(Prop.TeleportTo[ZeroPosition.GetTransform()]):
                        if(ProductType = B_TYPE):
                            if(set AllIsUsedBProps[Index] = false):
                        if(ProductType = CB_TYPE):
                            if(set AllIsUsedCBProps[Index] = false):
                        if(ProductType = SB_TYPE):
                            if(set AllIsUsedSBProps[Index] = false):
                        if(ProductType = PCS_TYPE):
                            if(set AllIsUsedPCSProps[Index] = false):
                        break
                    Sleep(0.04)
            else:
                PrintFatalError("On Animate Bake. GetAnimationController failed")
            if(ProductType = SB_TYPE):
                #Print("Trying to animate Sandwich make")
                spawn:
                    AnimateBake(BakeTransform, Speed, MaybeEffectSound, PCS_TYPE)
            
    <#
    NOT_USEDAnimateShip<public>(BakeTransform : transform, Speed : float, MaybeEffectSound : ?audio_player_device, ProductType : int )<suspends> : void =
        var Index : int = 0
        var Prop : creative_prop = creative_prop{}
        var RotationApplied1 : rotation = rotation{}
        var RotationApplied2 : rotation = rotation{}
        var RotationApplied3 : rotation = rotation{}
        var HalfSize : vector3 = vector3{X:=0.5,Y:=0.5,Z:=0.5}
        var NormalSize : vector3 = vector3{X:=1.0,Y:=1.0,Z:=1.0}
        var DoubleSize : vector3 = vector3{X:=2.0,Y:=2.0,Z:=2.0}
        var VectorToMidPoint : vector3 = vector3{}
        var VectorToEndPoint : vector3 = vector3{}
        var StartPoint : vector3 = vector3{}
        PropPoint := Prop.GetTransform().Translation
        VectorZero := vector3{X:=0.0,Y:=0.0,Z:=0.0}
        VectorSmallDecrement := vector3{X:=0.0,Y:=0.0,Z:=-32.0}

        case(ProductType):
            0=> 
                set RotationApplied1 = MakeRotationFromYawPitchRollDegrees(BSHIP1_RotationYaw, BSHIP1_RotationPitch,BSHIP1_RotationRoll)
                set RotationApplied2 = MakeRotationFromYawPitchRollDegrees(BSHIP2_RotationYaw, BSHIP2_RotationPitch,BSHIP2_RotationRoll)
                set RotationApplied3 = MakeRotationFromYawPitchRollDegrees(BSHIP3_RotationYaw, BSHIP3_RotationPitch,BSHIP3_RotationRoll)
                set StartPoint = BakeTransform.Translation + BSHIP_STARTPOSITION
                set VectorToMidPoint = (BakeTransform.Translation + BSHIP_MIDPOSITION)-StartPoint
                set VectorToEndPoint = (BakeTransform.Translation + BSHIP_ENDPOSITION)-(StartPoint+VectorToMidPoint) 
                loop:
                    if(Value := AllIsUsedBakedBProps[Index], not Value?, AuxProp := AllBakedBProps[Index], set Prop = AuxProp):
                        if(set AllIsUsedBakedBProps[Index] = true):
                            Log("Animating Ship. Bread selected: {Index}", LOG_LEVEL_BAKERS)
                        break
                    else:
                        set Index = Index + 1
                    if(Index > AllBakedBProps.Length):
                        #Print("Fatal error on AnimateBake. Bread to use as animation don´t found")
                        break
            1=> 
                set StartPoint = BakeTransform.Translation + CBSHIP_STARTPOSITION
                set VectorToMidPoint = (BakeTransform.Translation + CBSHIP_MIDPOSITION)-StartPoint
                set VectorToEndPoint = (BakeTransform.Translation + CBSHIP_ENDPOSITION)-(StartPoint+VectorToMidPoint)
                set RotationApplied1 = MakeRotationFromYawPitchRollDegrees(StandarSHIP_RotationYaw,StandarSHIP_RotationPitch,StandarSHIP_RotationRoll)
                set RotationApplied3 = MakeRotationFromYawPitchRollDegrees(CBSHIP3_RotationYaw, CBSHIP3_RotationPitch,CBSHIP3_RotationRoll)
                loop:
                    if(Value := AllIsUsedBakedCBProps[Index], not Value?, AuxProp := AllBakedCBProps[Index], set Prop = AuxProp):
                        if(set AllIsUsedBakedCBProps[Index]):
                        break
                    else:
                        set Index = Index + 1
                    if(Index > AllBakedCBProps.Length):
                        #Print("Fatal error on AnimateBake. Cheese Bread to use as animation don´t found")
                        break
            2=> 
                set StartPoint = BakeTransform.Translation + PCSSHIP_STARTPOSITION
                set VectorToMidPoint = (BakeTransform.Translation + PCSSHIP_MIDPOSITION)-StartPoint
                set VectorToEndPoint = (BakeTransform.Translation + PCSSHIP_ENDPOSITION)-(StartPoint+VectorToMidPoint)
                set RotationApplied1 = MakeRotationFromYawPitchRollDegrees(StandarSHIP_RotationYaw,StandarSHIP_RotationPitch,StandarSHIP_RotationRoll)
                set RotationApplied3 = MakeRotationFromYawPitchRollDegrees(PCSSHIP3_RotationYaw, PCSSHIP3_RotationPitch,PCSSHIP3_RotationRoll)
                loop:
                    if(Value := AllIsUsedPCSProps[Index], not Value?, AuxProp := AllPCSProps[Index], set Prop = AuxProp):
                        if(set AllIsUsedPCSProps[Index] = true):
                        break
                    else:
                        set Index = Index + 1
                    if(Index > AllPCSProps.Length):
                        #Print("Fatal error on AnimateBake. PCS to use as animation don´t found")
                        break
            _=> #Print("FATAL error on AnimateShip, ProductType invalid: {ProductType}")

        Log("Starting sending animation {Index}", LOG_LEVEL_BAKERS)
        if(Prop.TeleportTo[NewTransform:=transform{Translation := StartPoint, Rotation := RotationApplied1, Scale := HalfSize}]):

        StandarScale := BakeTransform.Scale

        TicksAnimationWait := 1.0
        TicksAnimationFrame1 := 1.0
        TicksAnimationFrame2 := 1.0
        TicksAnimationFrame3 := 2.0
        TicksAnimationWait2 :=  3.0
        TotalTicksOfAnimation := TicksAnimationWait + TicksAnimationFrame1 + TicksAnimationFrame2 + TicksAnimationFrame3 + TicksAnimationWait2

        AnimationWaitTime : float = (Speed/TotalTicksOfAnimation)*TicksAnimationWait
        Frame1Time : float = (5.0/TotalTicksOfAnimation)*TicksAnimationFrame1
        Frame2Time : float = (5.0/TotalTicksOfAnimation)*TicksAnimationFrame2
        Frame3Time : float = (5.0/TotalTicksOfAnimation)*TicksAnimationFrame3
        AnimationWaitTime2 : float = (Speed/TotalTicksOfAnimation)*TicksAnimationWait2

            MovementKeyFrame1 : keyframe_delta = keyframe_delta #Key Frame 1 to begining
            {
                DeltaLocation := VectorToMidPoint
                DeltaRotation := RotationApplied2
                DeltaScale := DoubleSize
                Time := Frame1Time
                Interpolation := Linear
            }
        <#
            MovementKeyFrame2 : keyframe_delta = keyframe_delta #Key Frame 1 to begining
            {
                DeltaLocation := VectorSmallDecrement
                DeltaRotation := RotationApplied3
                DeltaScale := NormalSize
                Time := Frame2Time
                Interpolation := Linear
            }
        #>
            MovementKeyFrame3 : keyframe_delta = keyframe_delta #Key Frame 1 to begining
            {
                DeltaLocation := VectorToEndPoint
                DeltaRotation := RotationApplied3
                DeltaScale := NormalSize
                Time := Frame3Time
                Interpolation := Linear
            }
    
            if(PropController := Prop.GetAnimationController[]):
                # Make an array of key frames
                FortVaderKeyFrames :  []keyframe_delta = array { MovementKeyFrame1 <#, MovementKeyFrame2#>, MovementKeyFrame3}
                # Set the animation to play
                PropController.SetAnimation(FortVaderKeyFrames, ?Mode:=animation_mode.OneShot)
                # Play Animation
                Sleep(AnimationWaitTime)
                PropController.Play()
                if(ValidEffectSound := MaybeEffectSound?):
                    ValidEffectSound.Play()
                loop:
                    State := PropController.GetState()
                    if(State = animation_controller_state.Stopped):
                        if(Prop.TeleportTo[ZeroPosition.GetTransform()]):
                        if(ProductType = B_TYPE):
                            if(set AllIsUsedBakedBProps[Index] = false):
                        if(ProductType = CB_TYPE):
                            if(set AllIsUsedBakedCBProps[Index] = false):
                        if(ProductType = SB_TYPE):
                            if(set AllIsUsedBakedSBProps[Index] = false):
                        if(ProductType = PCS_TYPE):
                            if(set AllIsUsedPCSProps[Index] = false):
                        break
                    Sleep(0.04)#>

AnimatePropAppearance<public>(Prop : creative_prop, FinalTransform : transform, MaybeEffectSound : ?audio_player_device)<suspends> : void =

    PropPoint := Prop.GetTransform().Translation

    DestinyPoint := FinalTransform.Translation
    BasePoint := DestinyPoint - vector3{X:=0.0,Y:=0.0,Z:=512.0}
    BouncePoint := DestinyPoint + vector3{X:=0.0,Y:=0.0,Z:=128.0}

    X:=1.0+0.2

    BounceEffectScale := vector3{X:=0.8,Y:=0.8,Z:=1.25} #Scale is multiplicative!
    BounceDesEffectScale := vector3{X:=1.0/BounceEffectScale.X,Y:=1.0/BounceEffectScale.Y,Z:=1.0/BounceEffectScale.Z}
    NormalScale := Prop.GetTransform().Scale
    NormalRotation := FinalTransform.Rotation

    MovementKeyFrame1 : keyframe_delta = keyframe_delta #Key Frame 1 to base position

    {
        DeltaLocation := BasePoint - PropPoint
        DeltaRotation := NormalRotation
        DeltaScale := NormalScale
        Time := 0.1
        Interpolation := EaseIn
    }

    MovementKeyFrame2 : keyframe_delta = keyframe_delta #Key to BouncePoint with Bounce Scale
    {
        DeltaLocation := BouncePoint - BasePoint 
        DeltaRotation := NormalRotation
        DeltaScale := BounceEffectScale
        Time := 0.2
        Interpolation := Linear
    }

    MovementKeyFrame3 : keyframe_delta = keyframe_delta #Key Frame 3 to go to the Destiny
    {
        DeltaLocation := DestinyPoint - BouncePoint
        DeltaRotation := NormalRotation
        DeltaScale := BounceDesEffectScale
        Time := 0.4
        Interpolation := EaseOut
    }

    if(PropController := Prop.GetAnimationController[]):
        #Print("Playing cool animation")
        # Make an array of key frames
        FortVaderKeyFrames :  []keyframe_delta = array { MovementKeyFrame1 , MovementKeyFrame2,  MovementKeyFrame3}
        # Set the animation to play
        PropController.SetAnimation(FortVaderKeyFrames, ?Mode:=animation_mode.OneShot)
        # Play Animation
        PropController.Play()
        if(ValidEffectSound := MaybeEffectSound?):
            ValidEffectSound.Play()
        loop:
            State := PropController.GetState()
            if(State = animation_controller_state.Stopped):
                break
            Sleep(0.04)

AnimateDeposit<public>(Prop : creative_prop, MaxValue : float, CurrentValue : float, MaybeEffectSound : ?audio_player_device)<suspends> : void =
    
    DepositHeightRange := MaxDepositHeight - MinDepositHeight
    HeightFromBottom := (DepositHeightRange/MaxValue)*CurrentValue
    CurrentValueHeight := HeightFromBottom + MinDepositHeight

    #Log("MaxDepositHeight: {MaxDepositHeight}, CurrentValue: {CurrentValue}", LOG_LEVEL_BAKERS)

    DestinyPointX := Prop.GetTransform().Translation.X
    DestinyPointY := Prop.GetTransform().Translation.Y
    DestinyPointZ := CurrentValueHeight

    DestinyTranslation:=vector3{X:=DestinyPointX,Y:=DestinyPointY,Z:=DestinyPointZ}
    DestinyScale:=vector3{X:=1.0,Y:=1.0,Z:=1.0}
    DestinyRotation:=Prop.GetTransform().Rotation

    BounceEffectScale := vector3{X:=0.8,Y:=0.8,Z:=1.25} #Scale is multiplicative!
    BounceDesEffectScale := vector3{X:=1.0/BounceEffectScale.X,Y:=1.0/BounceEffectScale.Y,Z:=1.0/BounceEffectScale.Z}
    NormalScale := NeutralScale
    NormalRotation : rotation = rotation{}

    MovementKeyFrame1 : keyframe_delta = keyframe_delta #Key Frame 1 to base position
    {
        DeltaLocation := DestinyTranslation - Prop.GetTransform().Translation
        DeltaRotation := NormalRotation
        DeltaScale := NormalScale
        Time := 0.5
        Interpolation := EaseInOut
    }

    if(PropController := Prop.GetAnimationController[]):
        # Make an array of key frames
        FortVaderKeyFrames :  []keyframe_delta = array { MovementKeyFrame1 }
        # Set the animation to play
        PropController.SetAnimation(FortVaderKeyFrames, ?Mode:=animation_mode.OneShot)
        # Play Animation
        PropController.Play()
        if(ValidEffectSound := MaybeEffectSound?):
            ValidEffectSound.Play()
        loop:
            State := PropController.GetState()
            if(State = animation_controller_state.Stopped):
                break
            Sleep(0.04)
    else:
        Log("Fatal error trying to animate deposit",LOG_LEVEL_1DETAILING)

    

            
