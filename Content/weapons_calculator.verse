
#DUNGEON PERKS
MainFormula_DelayDungeon(LV : int) : float = 
    TotalDelay:=Dungeon_DelayTime_Base - (Dungeon_DelayTime_ImprovementBase*LV)
    if(Dungeon_DelayTime_MinTime>TotalDelay):
        return Dungeon_DelayTime_MinTime
    else:
        return TotalDelay

MainFormula_MaxLVDelayDungeon() : int = 
    var MaxLV : int = 0
    if(Value:=Floor[(Dungeon_DelayTime_Base-Dungeon_DelayTime_MinTime) / Dungeon_DelayTime_ImprovementBase]):
        set MaxLV = Value
    return MaxLV

#HEXA PERKS
MainFormula_DelayHexaGame(LV : int) : float = 
    TotalDelay:=HexaGame_DelayTime_Base - (HexaGame_DelayTime_ImprovementBase*LV)
    if(HexaGame_DelayTime_MinTime>TotalDelay):
        return HexaGame_DelayTime_MinTime
    else:
        return TotalDelay

MainFormula_MaxLVDelayHexaGame() : int = 
    var MaxLV : int = 0
    if(Value:=Floor[(HexaGame_DelayTime_Base-HexaGame_DelayTime_MinTime) / HexaGame_DelayTime_ImprovementBase]):
        set MaxLV = Value
    return MaxLV

#Evolution
MainFormula_WeaponEvolution(Type : int, Tier : int) : float =
    var Amount : float = 0.0
    set Amount = 500.0*Tier
    return Amount

MainFormula_WeaponEvolutionCost(Type : int, Tier : int) : float =
    var Amount : float = 0.0
    var Base : float = 0.0
    if(Type=BOW_TYPE):
        set Amount = Pow(550.0,(Tier-1)*1.0)
    else if(Type=SWORD_TYPE):
        set Amount = Pow(25.0,Tier*1.0)
    return Amount

#Damage
MainFormula_WeaponDamage(Type : int, Tier : int, LV : int) : float =
    var Amount : float = 0.0
    if(Type=SWORD_TYPE):
        set Amount = SwordBaseDamage * LV * (5*Tier)
    else if(Type=BOW_TYPE):
        set Amount = (BowBaseDamage * LV * (5*Tier))+30.0
    return Amount

MainFormula_WeaponDamageCost(Type : int, LV : int) : float =
    var Amount : float = 0.0
    var Base : float = 0.0
    if(Type=SWORD_TYPE):
        set Base = SwordBaseDamageCost
    else if(Type=BOW_TYPE):
        set Base = BowBaseDamageCost
    else:
        set Base = 99.9
    return Pow(Base,LV*1.0) 

MainFormula_EvolveDungeonCost(HBHardBreadLV : int, LV : int) : float =
    #AVGHits := (MaxNum_AVGHit + MinNum_AVGHit)/2.0
    #CostPerLevel:=CheckFinalMutantMaterialAward(HBHardBreadLV,AVGHits,1.0)*75.0
    CostPerLevel:=HBHardBreadLV

    return CostPerLevel*(LV*1.0)


#Critical Prob
MainFormula_WeaponCriticalProb(Type : int, LV : int) : float =
    var Base : float = 0.0
    if(Type=SWORD_TYPE):
        set Base = SwordBaseCriticalProb
    else if(Type=BOW_TYPE):
        set Base = BowBaseCriticalProb
    else:
        set Base = 99.9
    return Base*LV

MainFormula_WeaponCriticalProbCost(Type : int, LV : int) : float =
    var Amount : float = 0.0
    var Base : float = 0.0
    if(Type=SWORD_TYPE):
        set Base = SwordBaseCriticalProbCost
    else if(Type=BOW_TYPE):
        set Base = BowBaseCriticalProbCost
    else:
        set Base = 99.9
    return Pow(Base,LV*1.0) 

#Critical
MainFormula_WeaponCritical(Type : int, LV : int) : float =
    var Base : float = 0.0
    if(Type=SWORD_TYPE):
        set Base = SwordBaseCritical
    else if(Type=BOW_TYPE):
        set Base = BowBaseCritical
    else:
        set Base = 99.9
    return Base*LV

MainFormula_WeaponCriticalCost(Type : int, LV : int) : float =
    var Amount : float = 0.0
    var Base : float = 0.0
    if(Type=SWORD_TYPE):
        set Base = SwordBaseCriticalCost
    else if(Type=BOW_TYPE):
        set Base = BowBaseCriticalCost
    else:
        set Base = 99.9
    return Pow(Base,LV*1.0) 

#Reward
MainFormula_WeaponReward(Type : int, Tier : int ,LV : int) : float =
    var Base : float = 0.0
    if(Type=SWORD_TYPE):
        set Base = SwordBaseReward
    else if(Type=BOW_TYPE):
        set Base = BowBaseReward
    else:
        set Base = 99.9
    return Base*LV*Tier

MainFormula_WeaponRewardCost(Type : int, LV : int) : float =
    var Amount : float = 0.0
    var Base : float = 0.0
    if(Type=SWORD_TYPE):
        set Base = SwordBaseRewardCost
    else if(Type=BOW_TYPE):
        set Base = BowBaseRewardCost
    else:
        set Base = 99.9
    return Pow(Base,LV*1.0) 

#Calcule Weapon Reward
CalculeMobReward(AreaLV : int, MobType : int, NumDungeon : int) : tuple(int,float) =
    var AwardAmount : float = 0.0
    var AwardType : int = MUTANTMATERIAL
    if(NumDungeon<>-1):
        if(MobType=MINIBOSS_MOB):
            set AwardAmount = AreaLV*Reward_NormalMobMultiplier*Reward_MiniBossMultiplier
        else if(MobType=GROUND_MOB):
            set AwardAmount = AreaLV*Reward_DungeonMobMultiplier*Reward_GroundMobMultiplier
        else if(MobType=DRUID_MOB):
            set AwardAmount = AreaLV*Reward_DungeonMobMultiplier*Reward_GroundMobMultiplier
    else:
        if(MobType=MINIBOSS_MOB):
            set AwardAmount = AreaLV*Reward_NormalMobMultiplier*Reward_MiniBossMultiplier
        else if(MobType=GROUND_MOB):
            set AwardAmount = AreaLV*Reward_NormalMobMultiplier*Reward_GroundMobMultiplier
        else if(MobType=DRUID_MOB):
            set AwardAmount = AreaLV*Reward_NormalMobMultiplier*Reward_GroundMobMultiplier
    return (AwardType,AwardAmount)

<#
MainFormula_WeaponDamageCost(Purity : int) : float = 
    return MainFormula_PetPurityCost(Purity)#>

#Calcule mob health
CalculeMobHealth<public>(MobType : int, ExpectedHammerDamage : float, HBLevelArea : int) : float =
    var TotalHealth : float = 99999999999999999.9
    #var Base : float = 99.9
    var NumHits : float = 99.0
    if(MobType=PLANT_MOB):
        #set Base = PlantMobBaseHealth
        set NumHits = PlantMobBaseNumHits
    else if(MobType=GROUND_MOB):
        #set Base = GroundMobBaseHealth
        set NumHits = GroundMobBaseNumHits
    else if(MobType=DRUID_MOB):
        #set Base = DruidMobBaseHealth
        set NumHits = DruidMobBaseNumHits
    else if(MobType=MINIBOSS_MOB):
        #set Base = GroundMobBaseHealth*MiniBossMultiplier
        set NumHits = MiniBossMultiplier

    set TotalHealth = ExpectedHammerDamage*NumHits

    if(HBLevelArea>7): #Add extra difficulty 
        set TotalHealth = TotalHealth + TotalHealth*0.25*(HBLevelArea-7)
    #Multiply by num hits hits
    
    return TotalHealth


#Calcule Weapon Stat
CalculeWeaponMaxLVStat<public>(StatToCalcule : int, WeaponType : int) : int =
    if(StatToCalcule=WEAPONSTAT_TYPE_EVOLUTION):
        if(WeaponType=SWORD_TYPE):
            return 4
        else:
            return 3
    return WEAPON_STATMAXLEVEL

GetHitMobDamage(NumDungeon : int) : float =
    return (NumDungeon*5.0) + 5.0

GetExpectedWeaponDamageAVG(SapphiresInvest : float, MutantMaterialInvest : float) : float =
    Log("SapphiresInvest: {SapphiresInvest}, MutantMaterialInvest: {MutantMaterialInvest}",LOG_LEVEL_DEBUG7)
    var WeaponTier : int = 1
    var MaxLevel : int = CalculeWeaponMaxLVStat(WEAPONSTAT_TYPE_EVOLUTION,SWORD_TYPE)
    loop:
        MutantNextCost:=CalculeWeaponStat(WEAPONSTAT_TYPE_EVOLUTION,SWORD_TYPE,WeaponTier,1)(1)
        if(MutantMaterialInvest>=MutantNextCost):
            set WeaponTier = WeaponTier + 1
        else:
            break
        if(WeaponTier>MaxLevel):
            set WeaponTier = MaxLevel
            break
    #Damage
    var DamageLV : int = 1
    set MaxLevel = CalculeWeaponMaxLVStat(WEAPONSTAT_TYPE_DAMAGE,SWORD_TYPE)
    loop:
        SapphiresNextCost:=CalculeWeaponStat(WEAPONSTAT_TYPE_DAMAGE,SWORD_TYPE,WeaponTier,DamageLV)(1)
        if(SapphiresInvest>=SapphiresNextCost):
            set DamageLV = DamageLV + 1
        else:
            break
        if(DamageLV>MaxLevel):
            set DamageLV = MaxLevel
            break
    #Critical Prob
    var CriticalProbLV : int = 1
    set MaxLevel = CalculeWeaponMaxLVStat(WEAPONSTAT_TYPE_CRITICALPROB,SWORD_TYPE)
    loop:
        SapphiresNextCost:=CalculeWeaponStat(WEAPONSTAT_TYPE_CRITICALPROB,SWORD_TYPE,WeaponTier,CriticalProbLV)(1)
        if(SapphiresInvest>=SapphiresNextCost):
            set CriticalProbLV = CriticalProbLV + 1
        else:
            break
        if(CriticalProbLV>MaxLevel):
            set CriticalProbLV = MaxLevel
            break
    #Critical
    var CriticalLV : int = 1
    set MaxLevel = CalculeWeaponMaxLVStat(WEAPONSTAT_TYPE_CRITICALDAMAGE,SWORD_TYPE)
    loop:
        SapphiresNextCost:=CalculeWeaponStat(WEAPONSTAT_TYPE_CRITICALDAMAGE,SWORD_TYPE,WeaponTier,CriticalLV)(1)
        if(SapphiresInvest>=SapphiresNextCost):
            set CriticalLV = CriticalLV + 1
        else:
            break
        if(CriticalLV>MaxLevel):
            set CriticalLV = MaxLevel
            break
    #########
    Damage_Sword := CalculeWeaponStat(WEAPONSTAT_TYPE_DAMAGE,SWORD_TYPE,WeaponTier,DamageLV)(0)
    CriticalDamage_Sword := CalculeWeaponStat(WEAPONSTAT_TYPE_CRITICALDAMAGE,SWORD_TYPE,WeaponTier,CriticalLV)(0)
    CriticalProb_Sword := CalculeWeaponStat(WEAPONSTAT_TYPE_CRITICALPROB,SWORD_TYPE,WeaponTier,CriticalProbLV)(0)

    CriticalDamage := Damage_Sword * CriticalDamage_Sword * CriticalProb_Sword
    TotalDamage := Damage_Sword + CriticalDamage

    Log("WeaponTier: {WeaponTier} DamageLV: {DamageLV}, CriticalLV: {CriticalLV}, CriticalProbLV: {CriticalProbLV}",LOG_LEVEL_DEBUG7)
    #Log("Damage: {Damage_Sword}, CriticalDamage_Sword: {CriticalDamage_Sword}, CriticalProb_Sword: {CriticalProb_Sword}",LOG_LEVEL_DEBUG7)
    Log("TotalDamage: {TotalDamage}, CriticalDamage: {CriticalDamage}",LOG_LEVEL_DEBUG7)

    return TotalDamage

    
#Calcule Weapon Stat
CalculeWeaponStat<public>(StatToCalcule : int, WeaponType : int, Tier : int, LV : int) : tuple(float,float,float,float) =
    #### Next stats
    var ObjetiveStatLV : int = LV+1

    var CurrentStat : float = 0.0
    var NextStat : float = 0.0
    var CurrentCost : float = 0.0
    var NextCost : float = 0.0
    if(StatToCalcule=WEAPONSTAT_TYPE_EVOLUTION):
        ### Stats
        set CurrentStat = MainFormula_WeaponEvolution(WeaponType, Tier)
        set NextStat = MainFormula_WeaponEvolution(WeaponType, Tier+1)
        #### Costs
        set CurrentCost = MainFormula_WeaponEvolutionCost(WeaponType, Tier)
        set NextCost = MainFormula_WeaponEvolutionCost(WeaponType, Tier+1)
    else if(StatToCalcule = WEAPONSTAT_TYPE_DAMAGE):
        ### Stats
        set CurrentStat = MainFormula_WeaponDamage(WeaponType, Tier, LV)
        set NextStat = MainFormula_WeaponDamage(WeaponType, Tier, ObjetiveStatLV)
        #### Costs
        set CurrentCost = MainFormula_WeaponDamageCost(StatToCalcule, LV)
        set NextCost = MainFormula_WeaponDamageCost(StatToCalcule, ObjetiveStatLV)
    else if(StatToCalcule = WEAPONSTAT_TYPE_CRITICALPROB):
       ### Stats
       set CurrentStat = MainFormula_WeaponCriticalProb(WeaponType, LV)
       set NextStat = MainFormula_WeaponCriticalProb(WeaponType, ObjetiveStatLV)
       #### Costs
       set CurrentCost = MainFormula_WeaponCriticalProbCost(StatToCalcule, LV)
       set NextCost = MainFormula_WeaponCriticalProbCost(StatToCalcule, ObjetiveStatLV)
    else if(StatToCalcule = WEAPONSTAT_TYPE_CRITICALDAMAGE):
        ### Stats
        set CurrentStat = MainFormula_WeaponCritical(WeaponType, LV)
        set NextStat = MainFormula_WeaponCritical(WeaponType, ObjetiveStatLV)
        #### Costs
        set CurrentCost = MainFormula_WeaponCriticalCost(StatToCalcule, LV)
        set NextCost = MainFormula_WeaponCriticalCost(StatToCalcule, ObjetiveStatLV)
    else if(StatToCalcule = WEAPONSTAT_TYPE_REWARD):
        ### Stats
        set CurrentStat = MainFormula_WeaponReward(WeaponType, Tier, LV)
        set NextStat = MainFormula_WeaponReward(WeaponType, Tier, ObjetiveStatLV)
        #### Costs
        set CurrentCost = MainFormula_WeaponRewardCost(StatToCalcule, LV)
        set NextCost = MainFormula_WeaponRewardCost(StatToCalcule, ObjetiveStatLV)

    #### Difference
    TotalDifference := NextStat-CurrentStat
    return(CurrentStat,NextCost,TotalDifference,CurrentCost)



#DUNGEON TOKEN STAT
CalculeDungeonTokenStat<public>(StatToCalcule : int, CurrentLV : int) : tuple(float,float,float) =
    #### Next stats
    var ObjetiveStatLV : int = CurrentLV+1

    var CurrentStat : float = 0.0
    var NextStat : float = 0.0
    var CurrentCost : float = 0.0
    var NextCost : float = 0.0

    if(StatToCalcule=DUNGEONPERK_DELAY):
         ### Stats
         set CurrentStat = Dungeon_DelayTime_Base-(CurrentLV * Dungeon_DelayTime_ImprovementBase)
         set NextStat =  Dungeon_DelayTime_Base-(ObjetiveStatLV * Dungeon_DelayTime_ImprovementBase)
         #### Costs
         set CurrentCost = CurrentLV * DungeonPerk_DelayBaseCost
         set NextCost = ObjetiveStatLV * DungeonPerk_DelayBaseCost

    else if(StatToCalcule=DUNGEONPERK_ACCUMULATED):
        ### Stats
         set CurrentStat = CurrentLV * 1.0
         set NextStat =  ObjetiveStatLV * 1.0
         #### Costs
         set CurrentCost = (CurrentLV-1) * DungeonPerk_AccumulatedBaseCost
         set NextCost = (ObjetiveStatLV-1) * DungeonPerk_AccumulatedBaseCost
       
    #### Difference
    TotalDifference := NextStat-CurrentStat
    return(CurrentStat,NextCost,TotalDifference)


    #DUNGEON TOKEN STAT
CalculeHexaGameTokenStat<public>(StatToCalcule : int, CurrentLV : int) : tuple(float,float,float) =
    #### Next stats
    var ObjetiveStatLV : int = CurrentLV+1

    var CurrentStat : float = 0.0
    var NextStat : float = 0.0
    var CurrentCost : float = 0.0
    var NextCost : float = 0.0

    if(StatToCalcule=HEXAGAMEPERK_DELAY):
         ### Stats
         set CurrentStat = HexaGame_DelayTime_Base-(CurrentLV * HexaGame_DelayTime_ImprovementBase)
         set NextStat =  HexaGame_DelayTime_Base-(ObjetiveStatLV * HexaGame_DelayTime_ImprovementBase)
         #### Costs
         set CurrentCost = CurrentLV * HexaGamePerk_DelayBaseCost
         set NextCost = ObjetiveStatLV * HexaGamePerk_DelayBaseCost

    else if(StatToCalcule=HEXAGAMEPERK_ACCUMULATED):
        ### Stats
         set CurrentStat = (CurrentLV * 1.0) + (HexaGame_MinToken_Accumulated_MinusOne * 1.0)
         set NextStat =  (ObjetiveStatLV * 1.0) + (HexaGame_MinToken_Accumulated_MinusOne * 1.0)
         #### Costs
         set CurrentCost = CurrentLV * HexaGamePerk_AccumulatedBaseCost
         set NextCost = ObjetiveStatLV * HexaGamePerk_AccumulatedBaseCost
       
    #### Difference
    TotalDifference := NextStat-CurrentStat
    return(CurrentStat,NextCost,TotalDifference)
