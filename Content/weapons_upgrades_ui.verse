using { /UnrealEngine.com/Temporary/UI }
using { /UnrealEngine.com/Temporary/SpatialMath }
using { /Verse.org/Simulation }
using { /Fortnite.com/Devices }
using { /Fortnite.com/UI }
using { /Verse.org/Colors/NamedColors }
using { /Verse.org/Colors }
using { /Fortnite.com/Characters }

# See https://dev.epicgames.com/documentation/en-us/uefn/create-your-own-device-in-verse for how to create a verse device.

# A Verse-authored creative device that can be placed in a level
weapons_upgrades_ui := class():

    #Class settings
    var MaybeAgent : ?agent = false

    var LVTier_Sword : int = 0
    var Damage_Sword : float = 0.0
    var CriticalProb_Sword : float = 0.0
    var CriticalDamage_Sword : float = 0.0
    var Reward_Sword : float = 0.0

    var LVTier_Bow : int = 0
    var Damage_Bow : float = 0.0
    var CriticalProb_Bow: float = 0.0
    var CriticalDamage_Bow : float = 0.0
    var Reward_Bow : float = 0.0

    var LastTimeBowFired : float = 0.0
    var LastFiredPosition : transform = transform{}
    var LastPressedTime : float = 0.0

    var BakeryMachinesManager : bakery_machines_manager
    
    BackgroundImageSizeX : float = 3075.0
    BackgroundImageSizeY : float = 2100.0

    PTopColumn1SizeX : float = 350.0
    PTopColumn1SizeY : float = 80.0
    PTopColumn2SizeX : float = 135.0
    PTopColumn2SizeY : float = 80.0

    BUColumn1SizeX : float = 80.0
    BUColumn1SizeY : float = 80.0
    BUColumn2SizeX : float = 425.0
    BUColumn2SizeY : float = 80.0
    BUColumn3SizeX : float = 150.0
    BUColumn3SizeY : float = 80.0
    BUColumn4SizeX : float = 200.0
    BUColumn4SizeY : float = 80.0
    BUColumn5SizeX : float = 200.0
    BUColumn5SizeY : float = 80.0
    BUColumn6SizeX : float = 225.0
    BUColumn6SizeY : float = 80.0

    TitleTextBlock : text_block = text_block{DefaultTextColor:=White, DefaultShadowOffset:=ShadowOffset, DefaultShadowOpacity := DefaultTextShadowOpacity}

    WeaponSelected_Texture : texture_block = texture_block{DefaultDesiredSize:=vector2{X:=200.0,Y:=200.0}, DefaultImage:=Images.Sapphire_Icon}

    Sapphires_Texture : texture_block = texture_block{DefaultDesiredSize:=vector2{X:=IconSmallSizeX,Y:=IconSmallSizeY}, DefaultImage:=Images.Sapphire_Icon}
    TotalSapphires_BackgroundTexture : texture_block = texture_block{DefaultDesiredSize:=vector2{X:=SizeBoxSaphiresX,Y:=SizeBoxSaphiresY}, DefaultImage:=Images.Img_LineBackground2}
    
    TotalWater_BackgroundTexture : texture_block = texture_block{DefaultDesiredSize:=vector2{X:=SizeBoxSaphiresX,Y:=SizeBoxSaphiresY}, DefaultImage:=Images.Img_LineBackground2}
    TotalFlour_BackgroundTexture : texture_block = texture_block{DefaultDesiredSize:=vector2{X:=SizeBoxSaphiresX,Y:=SizeBoxSaphiresY}, DefaultImage:=Images.Img_LineBackground2}
    TotalCheese_BackgroundTexture : texture_block = texture_block{DefaultDesiredSize:=vector2{X:=SizeBoxSaphiresX,Y:=SizeBoxSaphiresY}, DefaultImage:=Images.Img_LineBackground2}
    TotalPimentos_BackgroundTexture : texture_block = texture_block{DefaultDesiredSize:=vector2{X:=SizeBoxSaphiresX,Y:=SizeBoxSaphiresY}, DefaultImage:=Images.Img_LineBackground2}
    TotalMutantMaterial_BackgroundTexture : texture_block = texture_block{DefaultDesiredSize:=vector2{X:=SizeBoxSaphiresX,Y:=SizeBoxSaphiresY}, DefaultImage:=Images.Img_LineBackground2}

    TotalSapphires_TextBlock : text_block = text_block{DefaultTextColor:=SapphiresTextColor,DefaultShadowOffset:=ShadowOffset, DefaultShadowOpacity := DefaultTextShadowOpacity}
    TotalMutantMaterial_TextBlock : text_block = text_block{DefaultTextColor:=MutantMaterialTextColor,DefaultShadowOffset:=ShadowOffset, DefaultShadowOpacity := DefaultTextShadowOpacity}
  
    PPU_Row1_Icon : texture_block = texture_block{DefaultDesiredSize:=vector2{X:=IconSizeX,Y:=IconSizeY},DefaultImage:=Images.Bakery_Knife_Icon}
    PPU_Row2_Icon : texture_block = texture_block{DefaultDesiredSize:=vector2{X:=IconSizeX,Y:=IconSizeY},DefaultImage:=Images.Bakery_Knife_Icon}
    PPU_Row3_Icon : texture_block = texture_block{DefaultDesiredSize:=vector2{X:=IconSizeX,Y:=IconSizeY},DefaultImage:=Images.Bakery_Knife_Icon}
    PPU_Row4_Icon : texture_block = texture_block{DefaultDesiredSize:=vector2{X:=IconSizeX,Y:=IconSizeY},DefaultImage:=Images.Bakery_Knife_Icon}
    PPU_Row5_Icon : texture_block = texture_block{DefaultDesiredSize:=vector2{X:=IconSizeX,Y:=IconSizeY},DefaultImage:=Images.Bakery_Knife_Icon}

    PPU_Row1_CurrentLV_TextBlock : text_block = text_block{DefaultTextColor:=CurrentTextColor, DefaultShadowOffset:=ShadowOffset, DefaultShadowOpacity := DefaultTextShadowOpacity}
    PPU_Row2_CurrentLV_TextBlock : text_block = text_block{DefaultTextColor:=CurrentTextColor, DefaultShadowOffset:=ShadowOffset, DefaultShadowOpacity := DefaultTextShadowOpacity}
    PPU_Row3_CurrentLV_TextBlock : text_block = text_block{DefaultTextColor:=CurrentTextColor, DefaultShadowOffset:=ShadowOffset, DefaultShadowOpacity := DefaultTextShadowOpacity}
    PPU_Row4_CurrentLV_TextBlock : text_block = text_block{DefaultTextColor:=CurrentTextColor, DefaultShadowOffset:=ShadowOffset, DefaultShadowOpacity := DefaultTextShadowOpacity}
    PPU_Row5_CurrentLV_TextBlock : text_block = text_block{DefaultTextColor:=CurrentTextColor, DefaultShadowOffset:=ShadowOffset, DefaultShadowOpacity := DefaultTextShadowOpacity}

    PPU_Row1_Current_TextBlock : text_block = text_block{DefaultTextColor:=CurrentTextColor, DefaultShadowOffset:=ShadowOffset, DefaultShadowOpacity := DefaultTextShadowOpacity}
    PPU_Row2_Current_TextBlock : text_block = text_block{DefaultTextColor:=CurrentTextColor, DefaultShadowOffset:=ShadowOffset, DefaultShadowOpacity := DefaultTextShadowOpacity}
    PPU_Row3_Current_TextBlock : text_block = text_block{DefaultTextColor:=CurrentTextColor, DefaultShadowOffset:=ShadowOffset, DefaultShadowOpacity := DefaultTextShadowOpacity}
    PPU_Row4_Current_TextBlock : text_block = text_block{DefaultTextColor:=CurrentTextColor, DefaultShadowOffset:=ShadowOffset, DefaultShadowOpacity := DefaultTextShadowOpacity}
    PPU_Row5_Current_TextBlock : text_block = text_block{DefaultTextColor:=CurrentTextColor, DefaultShadowOffset:=ShadowOffset, DefaultShadowOpacity := DefaultTextShadowOpacity}

    PPU_Row1_Upgrade_TextBlock : text_block = text_block{DefaultTextColor:=UpgradeTextColor, DefaultShadowOffset:=ShadowOffset, DefaultShadowOpacity := DefaultTextShadowOpacity}
    PPU_Row2_Upgrade_TextBlock : text_block = text_block{DefaultTextColor:=UpgradeTextColor, DefaultShadowOffset:=ShadowOffset, DefaultShadowOpacity := DefaultTextShadowOpacity}
    PPU_Row3_Upgrade_TextBlock : text_block = text_block{DefaultTextColor:=UpgradeTextColor, DefaultShadowOffset:=ShadowOffset, DefaultShadowOpacity := DefaultTextShadowOpacity}
    PPU_Row4_Upgrade_TextBlock : text_block = text_block{DefaultTextColor:=UpgradeTextColor, DefaultShadowOffset:=ShadowOffset, DefaultShadowOpacity := DefaultTextShadowOpacity}
    PPU_Row5_Upgrade_TextBlock : text_block = text_block{DefaultTextColor:=UpgradeTextColor, DefaultShadowOffset:=ShadowOffset, DefaultShadowOpacity := DefaultTextShadowOpacity}

    PPU_Row1_UpgradeCost_TextBlock : text_block = text_block{DefaultTextColor:=SapphiresTextColor, DefaultShadowOffset:=ShadowOffset, DefaultShadowOpacity := DefaultTextShadowOpacity}
    PPU_Row2_UpgradeCost_TextBlock : text_block = text_block{DefaultTextColor:=SapphiresTextColor, DefaultShadowOffset:=ShadowOffset, DefaultShadowOpacity := DefaultTextShadowOpacity}
    PPU_Row3_UpgradeCost_TextBlock : text_block = text_block{DefaultTextColor:=SapphiresTextColor, DefaultShadowOffset:=ShadowOffset, DefaultShadowOpacity := DefaultTextShadowOpacity}
    PPU_Row4_UpgradeCost_TextBlock : text_block = text_block{DefaultTextColor:=SapphiresTextColor, DefaultShadowOffset:=ShadowOffset, DefaultShadowOpacity := DefaultTextShadowOpacity}
    PPU_Row5_UpgradeCost_TextBlock : text_block = text_block{DefaultTextColor:=SapphiresTextColor, DefaultShadowOffset:=ShadowOffset, DefaultShadowOpacity := DefaultTextShadowOpacity}

    PPU_Row1Button : button_quiet = button_quiet{}
    PPU_Row2Button : button_quiet = button_quiet{}
    PPU_Row3Button : button_quiet = button_quiet{}
    PPU_Row4Button : button_quiet = button_quiet{}
    PPU_Row5Button : button_quiet = button_quiet{}

    var MaybeCanvas : ?canvas = false
    var UIIsDisplayed : logic = false

    var TimeElapsed : int = 0

    #Navigate
    NavigateLeftButton : button_regular = button_regular{}
    NavigateRightButton : button_regular = button_regular{}

    #Close
    CloseButton : button_loud = button_loud{}

    #Arrays
    var AllResourcersColor : []color = array{}

    var AllSFP_TopRows_TextBlock : []text_block = array{}
    var AllSFP_Resourcers_TextBlock : []text_block = array{}
    var AllSFP_Rows_IconTexture : []texture_block = array{}

    var AllWU_Rows_CurrentLV_TextBlock : []text_block = array{}
    var AllWU_Rows_Current_TextBlock : []text_block = array{}
    var AllWU_Rows_Upgrade_TextBlock : []text_block = array{}
    var AllWU_Rows_UpgradeCost_TextBlock : []text_block = array{}
    var AllWU_RowButtons : []button_quiet = array{}

    var CurrentPage : int = SWORD_TYPE

    InitUI() : void =
        SetArrays()
        InitSubscribers()

        Canvas := MakeWeaponsUpgradesCanvas()
        set MaybeCanvas = option{Canvas}
        CreateWeaponStats()

    SetArrowRegistration(InLastFiredPosition : transform, InLastTimeBowFired : float, InLastPressedTime : float): void =
        set LastFiredPosition = InLastFiredPosition
        set LastTimeBowFired = InLastTimeBowFired
        set LastPressedTime = InLastPressedTime

        CreateWeaponStats()

    SetArrays() : void =

        set AllSFP_Resourcers_TextBlock = AllSFP_Resourcers_TextBlock
        
        set AllWU_Rows_CurrentLV_TextBlock = array{PPU_Row1_CurrentLV_TextBlock,PPU_Row2_CurrentLV_TextBlock,PPU_Row3_CurrentLV_TextBlock,PPU_Row4_CurrentLV_TextBlock, PPU_Row5_CurrentLV_TextBlock}

        set AllWU_Rows_Current_TextBlock = array{PPU_Row1_Current_TextBlock,PPU_Row2_Current_TextBlock,PPU_Row3_Current_TextBlock,PPU_Row4_Current_TextBlock, PPU_Row5_Current_TextBlock}
        set AllWU_Rows_Upgrade_TextBlock = array{PPU_Row1_Upgrade_TextBlock,PPU_Row2_Upgrade_TextBlock,PPU_Row3_Upgrade_TextBlock,PPU_Row4_Upgrade_TextBlock, PPU_Row5_Upgrade_TextBlock}
        set AllWU_Rows_UpgradeCost_TextBlock = array{PPU_Row1_UpgradeCost_TextBlock,PPU_Row2_UpgradeCost_TextBlock,PPU_Row3_UpgradeCost_TextBlock,PPU_Row4_UpgradeCost_TextBlock, PPU_Row5_UpgradeCost_TextBlock}

        set AllWU_RowButtons = array{PPU_Row1Button,PPU_Row2Button,PPU_Row3Button,PPU_Row4Button,PPU_Row5Button} 

    InitSubscribers() : void =
        CloseButton.OnClick().Subscribe(Remove_Canvas) 
        NavigateLeftButton.OnClick().Subscribe(ChangeWeapon) 
        NavigateRightButton.OnClick().Subscribe(ChangeWeapon) 

        PPU_Row1Button.OnClick().Subscribe(handler_weaponsandmobs{MaybeWeaponsUpgradesUI:=option{Self}, WeaponPerk:=WEAPONSTAT_TYPE_EVOLUTION}.HandlerImproveWeapon)
        PPU_Row2Button.OnClick().Subscribe(handler_weaponsandmobs{MaybeWeaponsUpgradesUI:=option{Self}, WeaponPerk:=WEAPONSTAT_TYPE_DAMAGE}.HandlerImproveWeapon)
        PPU_Row3Button.OnClick().Subscribe(handler_weaponsandmobs{MaybeWeaponsUpgradesUI:=option{Self}, WeaponPerk:=WEAPONSTAT_TYPE_CRITICALPROB}.HandlerImproveWeapon)
        PPU_Row4Button.OnClick().Subscribe(handler_weaponsandmobs{MaybeWeaponsUpgradesUI:=option{Self}, WeaponPerk:=WEAPONSTAT_TYPE_CRITICALDAMAGE}.HandlerImproveWeapon)
        PPU_Row5Button.OnClick().Subscribe(handler_weaponsandmobs{MaybeWeaponsUpgradesUI:=option{Self}, WeaponPerk:=WEAPONSTAT_TYPE_REWARD}.HandlerImproveWeapon)

    TryImprovePerk(Perk : int) : void =
        if(Agent := MaybeAgent?,Player := player[Agent], Player.IsActive[], PlayerData:=BakeryMachinesManager.PlayerDataCache):
            var LVWeaponTier : int = 0
            var LVWeaponPerk : int = 0

            var TupleResult : tuple(float,float,float,float) = (0.0,0.0,0.0,0.0)
            var MaxLevel : int = 0

            if(CurrentPage = SWORD_TYPE):
                set LVWeaponTier = PlayerData.Tier_Sword
                if(Perk = WEAPONSTAT_TYPE_EVOLUTION):
                    set LVWeaponPerk = PlayerData.Tier_Sword 
                else if(Perk = WEAPONSTAT_TYPE_DAMAGE):
                    set LVWeaponPerk = PlayerData.LV_SwordDamage 
                else if(Perk = WEAPONSTAT_TYPE_CRITICALPROB):
                    set LVWeaponPerk = PlayerData.LV_SwordCriticalProb 
                else if(Perk = WEAPONSTAT_TYPE_CRITICALDAMAGE):
                    set LVWeaponPerk = PlayerData.LV_SwordCriticalDamage 
                else if(Perk = WEAPONSTAT_TYPE_REWARD):
                    set LVWeaponPerk = PlayerData.LV_SwordReward 
            else if(CurrentPage = BOW_TYPE):
                set LVWeaponTier = PlayerData.Tier_Bow
                if(Perk = WEAPONSTAT_TYPE_EVOLUTION):
                    set LVWeaponPerk = PlayerData.Tier_Bow 
                else if(Perk = WEAPONSTAT_TYPE_DAMAGE):
                    set LVWeaponPerk = PlayerData.LV_BowDamage 
                else if(Perk = WEAPONSTAT_TYPE_CRITICALPROB):
                    set LVWeaponPerk = PlayerData.LV_BowCriticalProb 
                else if(Perk = WEAPONSTAT_TYPE_CRITICALDAMAGE):
                    set LVWeaponPerk = PlayerData.LV_BowCriticalDamage 
                else if(Perk = WEAPONSTAT_TYPE_REWARD):
                    set LVWeaponPerk = PlayerData.LV_BowReward 

            set TupleResult = CalculeWeaponStat(Perk, CurrentPage, LVWeaponTier, LVWeaponPerk)
            set MaxLevel = CalculeWeaponMaxLVStat(Perk, CurrentPage)
            
            #Print("MaxLevel: {MaxLevel}, LVWeaponPerk: {LVWeaponPerk}")
            #Print("Cost {GetMagnitudeLetter(TupleResult(1),"","")}")

            if(Perk = WEAPONSTAT_TYPE_EVOLUTION):
                if(TupleResult(1)<=PlayerData.Inventory_MutantMaterial, MaxLevel>LVWeaponPerk):
                    TableWithoutMutantMaterial:=BakeryMachinesManager.GameManager.PlayersDataManager.WithdrawResource(PlayerData,TupleResult(1),MUTANTMATERIAL)
                    FinalTable:=BakeryMachinesManager.GameManager.PlayersDataManager.ImproveWeaponStat(TableWithoutMutantMaterial,Perk,CurrentPage)
                    #BakeryMachinesManager.GameManager.PlayersDataManager.SaveNewData(Player,PlayerData,FinalTable)

                    #BakeryMachinesManager.GameManager.PlayersDataManager.SaveAllNewData(Player,PlayerData,FinalTable, BakeryMachinesManager)
                    BakeryMachinesManager.SetNewPlayerDataCache(FinalTable)
                    BakeryMachinesManager.GameManager.GeneralMobsManager.RefreshSwordEffects()
                    spawn:
                        BakeryMachinesManager.GameManager.GeneralMobsManager.GranAgentWeapons(Agent)
                    BakeryMachinesManager.GameManager.VFXManager.BigObjetive_AudioPlayer.Play(Agent)
                    UpdateUI()
                    if(CurrentPage = SWORD_TYPE):
                        BakeryMachinesManager.CheckQuestProgress(1.0,WEAPONIMPROVE_TUTORIAL_TYPE)
            else if(TupleResult(1)<=PlayerData.Inventory_Sapphires, MaxLevel>LVWeaponPerk):
                TableWithoutSapphires:=BakeryMachinesManager.GameManager.PlayersDataManager.WithdrawResource(PlayerData,TupleResult(1),SAPPHIRES)
                FinalTable:=BakeryMachinesManager.GameManager.PlayersDataManager.ImproveWeaponStat(TableWithoutSapphires,Perk,CurrentPage)
                #BakeryMachinesManager.GameManager.PlayersDataManager.SaveAllNewData(Player,PlayerData,FinalTable, BakeryMachinesManager)
                BakeryMachinesManager.SetNewPlayerDataCache(FinalTable)
                #BakeryMachinesManager.GameManager.PlayersDataManager.SaveNewData(Player,PlayerData,FinalTable)
                if(CurrentPage = SWORD_TYPE):
                    BakeryMachinesManager.CheckQuestProgress(1.0,WEAPONIMPROVE_TUTORIAL_TYPE)
                UpdateUI()
                BakeryMachinesManager.GameManager.VFXManager.SmallObjetive_AudioPlayer.Play(Agent)
            
            else:
                BakeryMachinesManager.GameManager.VFXManager.AudioToSingle_Error.Play(Agent)
            CreateWeaponStats()
            BakeryMachinesManager.RefreshAllUI()
            UpdateUI()
            
    ChangeWeapon(Message : widget_message) : void =
        if(CurrentPage=SWORD_TYPE):
            set CurrentPage=BOW_TYPE
        else:
            set CurrentPage=SWORD_TYPE
        UpdateUI()

    SetAgent(InAgent : agent) : void =
        set UIIsDisplayed = false
        set MaybeAgent = option{InAgent}
        if(Character := InAgent.GetFortCharacter[]):
            SetArrowRegistration(Character.GetTransform(),-1.0,1.0)
        CreateWeaponStats()

    UpdateUI() : void =

        if(Agent := MaybeAgent?,Player := player[Agent], Player.IsActive[], PlayerData:=BakeryMachinesManager.PlayerDataCache):

            CloseButton.SetText(BasicMessage("CLOSE"))

            NavigateLeftButton.SetText(BasicMessage("◀ SWITCH WEAPON"))
            NavigateRightButton.SetText(BasicMessage("SWITCH WEAPON ▶"))

            #Inventory resourcers
            TotalSapphires_TextBlock.SetText(BasicMessage("{GetOptimizedMagnitudeLetter(PlayerData.Inventory_Sapphires)}"))
            TotalMutantMaterial_TextBlock.SetText(BasicMessage("{GetOptimizedMagnitudeLetter(PlayerData.Inventory_MutantMaterial)}"))

            for(Perk:=WEAPONSTAT_TYPE_EVOLUTION..WEAPONSTAT_TYPE_REWARD):
                if:
                    WU_Row_Current_TextBlock := AllWU_Rows_Current_TextBlock[Perk]
                    WU_Row_Upgrade_TextBlock := AllWU_Rows_Upgrade_TextBlock[Perk]
                    WU_Row_UpgradeCost_TextBlock := AllWU_Rows_UpgradeCost_TextBlock[Perk]
                then:
                    var TupleResult : tuple(float,float,float,float) = (0.0,0.0,0.0,0.0)
                    var MaxLevel : int = 0       

                    var WeaponStatLV : int = 0

                    var LVWeaponTier : int = 0
                    var LVWeaponDamage : int = 0
                    var LVWeaponCriticalProb : int = 0
                    var LVWeaponCriticalDamage : int = 0
                    var LVWeaponReward : int = 0
                    
                    if(CurrentPage = SWORD_TYPE):
                        if(PlayerData.Tier_Sword=1):
                            WeaponSelected_Texture.SetImage(Images.HammerTier_0)
                        else if(PlayerData.Tier_Sword=2):
                            WeaponSelected_Texture.SetImage(Images.HammerTier_1)
                        else if(PlayerData.Tier_Sword=3):
                            WeaponSelected_Texture.SetImage(Images.HammerTier_2)
                        else if(PlayerData.Tier_Sword=4):
                            WeaponSelected_Texture.SetImage(Images.HammerTier_3)
                        set LVWeaponTier = PlayerData.Tier_Sword 
                        set LVWeaponDamage = PlayerData.LV_SwordDamage 
                        set LVWeaponCriticalProb = PlayerData.LV_SwordCriticalProb
                        set LVWeaponCriticalDamage = PlayerData.LV_SwordCriticalDamage
                        set LVWeaponReward = PlayerData.LV_SwordReward
                    else if(CurrentPage = BOW_TYPE):
                        if(PlayerData.Tier_Bow=1):
                            WeaponSelected_Texture.SetImage(Images.BowTier_0)
                        else if(PlayerData.Tier_Bow=2):
                            WeaponSelected_Texture.SetImage(Images.BowTier_1)
                        else if(PlayerData.Tier_Bow=3):
                            WeaponSelected_Texture.SetImage(Images.BowTier_2)
                        set LVWeaponTier = PlayerData.Tier_Bow 
                        set LVWeaponDamage = PlayerData.LV_BowDamage 
                        set LVWeaponCriticalProb = PlayerData.LV_BowCriticalProb
                        set LVWeaponCriticalDamage = PlayerData.LV_BowCriticalDamage
                        set LVWeaponReward = PlayerData.LV_BowReward

                        
                    if(Perk = WEAPONSTAT_TYPE_EVOLUTION):
                        set TupleResult = CalculeWeaponStat(WEAPONSTAT_TYPE_EVOLUTION, CurrentPage, LVWeaponTier, LVWeaponTier)
                        set MaxLevel = CalculeWeaponMaxLVStat(WEAPONSTAT_TYPE_EVOLUTION, CurrentPage)
                        set WeaponStatLV = LVWeaponTier
                    else if(Perk = WEAPONSTAT_TYPE_DAMAGE):
                        set TupleResult = CalculeWeaponStat(WEAPONSTAT_TYPE_DAMAGE, CurrentPage, LVWeaponTier, LVWeaponDamage)
                        set MaxLevel = CalculeWeaponMaxLVStat(WEAPONSTAT_TYPE_DAMAGE, CurrentPage)
                        set WeaponStatLV = LVWeaponDamage
                    else if(Perk = WEAPONSTAT_TYPE_CRITICALPROB):
                        set TupleResult = CalculeWeaponStat(WEAPONSTAT_TYPE_CRITICALPROB, CurrentPage, LVWeaponTier, LVWeaponCriticalProb)
                        set MaxLevel = CalculeWeaponMaxLVStat(WEAPONSTAT_TYPE_CRITICALPROB, CurrentPage)
                        set WeaponStatLV = LVWeaponCriticalProb
                    else if(Perk = WEAPONSTAT_TYPE_CRITICALDAMAGE):
                        set TupleResult = CalculeWeaponStat(WEAPONSTAT_TYPE_CRITICALDAMAGE, CurrentPage, LVWeaponTier, LVWeaponCriticalDamage)
                        set MaxLevel = CalculeWeaponMaxLVStat(WEAPONSTAT_TYPE_CRITICALDAMAGE, CurrentPage)
                        set WeaponStatLV = LVWeaponCriticalDamage
                    else if(Perk = WEAPONSTAT_TYPE_REWARD):
                        set TupleResult = CalculeWeaponStat(WEAPONSTAT_TYPE_REWARD, CurrentPage, LVWeaponTier, LVWeaponReward)
                        set MaxLevel = CalculeWeaponMaxLVStat(WEAPONSTAT_TYPE_REWARD, CurrentPage)
                        set WeaponStatLV = LVWeaponReward

                    if(WeaponStatLV<MaxLevel):
                        if(Perk = WEAPONSTAT_TYPE_CRITICALPROB or Perk = WEAPONSTAT_TYPE_REWARD):
                            WU_Row_UpgradeCost_TextBlock.SetText(BasicMessage(GetMagnitudeLetter(TupleResult(1),"x ","")))
                            WU_Row_Upgrade_TextBlock.SetText(BasicMessage(GetMagnitudeLetter(TupleResult(2)*100.0,"+"," %")))
                            WU_Row_Current_TextBlock.SetText(BasicMessage(GetMagnitudeLetter(TupleResult(0)*100.0,"","%")))

                        else if(Perk = WEAPONSTAT_TYPE_CRITICALDAMAGE):
                            WU_Row_UpgradeCost_TextBlock.SetText(BasicMessage(GetMagnitudeLetter(TupleResult(1),"x ","")))
                            WU_Row_Upgrade_TextBlock.SetText(BasicMessage(GetMagnitudeLetter(TupleResult(2),"+","")))
                            WU_Row_Current_TextBlock.SetText(BasicMessage(GetMagnitudeLetter(TupleResult(0),"x ","")))

                        else if(Perk = WEAPONSTAT_TYPE_DAMAGE):
                            WU_Row_UpgradeCost_TextBlock.SetText(BasicMessage(GetMagnitudeLetter(TupleResult(1),"x ","")))
                            WU_Row_Upgrade_TextBlock.SetText(BasicMessage(GetMagnitudeLetter(TupleResult(2),"+","")))
                            WU_Row_Current_TextBlock.SetText(BasicMessage(GetMagnitudeLetter(TupleResult(0),"","")))

                        else if(Perk = WEAPONSTAT_TYPE_EVOLUTION):
                            WU_Row_UpgradeCost_TextBlock.SetText(BasicMessage(GetMagnitudeLetter(TupleResult(1),"x ","")))
                            WU_Row_Upgrade_TextBlock.SetText(BasicMessage(GetMagnitudeLetter(TupleResult(2)*100.0,"+"," %")))
                            WU_Row_Current_TextBlock.SetText(BasicMessage(GetMagnitudeLetter(TupleResult(0)*100.0,"","%")))


                        WU_Row_Current_TextBlock.SetTextColor(CurrentTextColor)
                        WU_Row_Upgrade_TextBlock.SetTextColor(UpgradeTextColor)
                    
                        if(Perk = WEAPONSTAT_TYPE_EVOLUTION):
                            if(TupleResult(1)<=PlayerData.Inventory_MutantMaterial):
                                WU_Row_UpgradeCost_TextBlock.SetTextColor(MutantMaterialTextColor)
                            else:
                                WU_Row_UpgradeCost_TextBlock.SetTextColor(NotEnoughTextColor)
                        else if(TupleResult(1)<=PlayerData.Inventory_Sapphires):
                            WU_Row_UpgradeCost_TextBlock.SetTextColor(SapphiresTextColor)
                        else:
                            WU_Row_UpgradeCost_TextBlock.SetTextColor(NotEnoughTextColor) 
                    else:
                        if(Perk = WEAPONSTAT_TYPE_CRITICALPROB or Perk = WEAPONSTAT_TYPE_REWARD):
                            WU_Row_Current_TextBlock.SetText(BasicMessage(GetMagnitudeLetter(TupleResult(0),"","%")))
                        else if(Perk = WEAPONSTAT_TYPE_CRITICALDAMAGE):
                            WU_Row_Current_TextBlock.SetText(BasicMessage(GetMagnitudeLetter(TupleResult(0),"x ","")))
                        else if(Perk = WEAPONSTAT_TYPE_DAMAGE):
                            WU_Row_Current_TextBlock.SetText(BasicMessage(GetMagnitudeLetter(TupleResult(0),"","")))
                        else if(Perk = WEAPONSTAT_TYPE_EVOLUTION):
                            WU_Row_Current_TextBlock.SetText(BasicMessage(GetMagnitudeLetter(TupleResult(0),"","")))

                        WU_Row_Upgrade_TextBlock.SetText(BasicMessage("MAXED"))
                        WU_Row_Upgrade_TextBlock.SetTextColor(MaxedColor)
                        WU_Row_UpgradeCost_TextBlock.SetText(BasicMessage("MAXED"))
                        WU_Row_UpgradeCost_TextBlock.SetTextColor(MaxedColor)
            WeaponSelected_Texture.SetDesiredSize(vector2{X:=200.0,Y:=200.0})

    Remove_Canvas(Message : widget_message) : void =
        set UIIsDisplayed = false
        if:
            Agent:=MaybeAgent?
            Player:=player[Agent]
            PlayerUI:=GetPlayerUI[Player]
            Canvas := MaybeCanvas?
        then:
            PlayerUI.RemoveWidget(Canvas)
            RemoveDelay_OnRemove()

    Remove_CanvasWithoutMessage() : void =
        set UIIsDisplayed = false
        #Print("Trying to remove UI")
        if:
            Agent:=MaybeAgent?
            Player:=player[Agent]
            PlayerUI:=GetPlayerUI[Player]
            Canvas := MaybeCanvas?
        then:
            #Print("Removing UI")
            PlayerUI.RemoveWidget(Canvas)
            RemoveDelay_OnRemove()

    RemoveDelay_OnAdd() : void =
        TitleTextBlock.SetText(BasicMessage("Weapon upgrades"))
        #CurrentBuyOption_textblock.SetVisibility(widget_visibility.Hidden)

    RemoveDelay_OnRemove() : void =
        TitleTextBlock.SetText(BasicMessage("Weapon upgrades "))
        #CurrentBuyOption_textblock.SetVisibility(widget_visibility.Hidden)

            
    Display_Canvas(WeaponTYPE : int) : void =
        set CurrentPage=WeaponTYPE
        if(not UIIsDisplayed?):
            set TimeElapsed = 0
            set UIIsDisplayed = true
            #spawn: 
            #    StartTimeElapsed()
            if:
                Agent:=MaybeAgent?
                PlayerUI:=GetPlayerUI[player[Agent]]
                Canvas := MaybeCanvas?
            then:
                ##Print("Assigning created UI before")
                PlayerUI.AddWidget(Canvas,player_ui_slot{InputMode := ui_input_mode.All, ZOrder:=1}) 
                RemoveDelay_OnAdd() 
                UpdateUI()
        
    
    CreateWeaponStats() : void =
        if(Agent := MaybeAgent?, Player := player[Agent], Player.IsActive[], PlayerData := BakeryMachinesManager.PlayerDataCache):
            #Sword
            set LVTier_Sword = PlayerData.Tier_Sword
            set Damage_Sword = CalculeWeaponStat(WEAPONSTAT_TYPE_DAMAGE,SWORD_TYPE,LVTier_Sword,PlayerData.LV_SwordDamage)(0)
            set CriticalDamage_Sword = CalculeWeaponStat(WEAPONSTAT_TYPE_CRITICALDAMAGE,SWORD_TYPE,LVTier_Sword,PlayerData.LV_SwordCriticalDamage)(0)
            set CriticalProb_Sword = CalculeWeaponStat(WEAPONSTAT_TYPE_CRITICALPROB,SWORD_TYPE,LVTier_Sword,PlayerData.LV_SwordCriticalProb)(0)
            #set Reward_Sword = CalculeWeaponStat(WEAPONSTAT_TYPE_REWARD,SWORD_TYPE,LVTier_Sword,PlayerData.LV_SwordReward)(0)
            #Bow
            set LVTier_Bow = PlayerData.Tier_Bow
            set Damage_Bow = CalculeWeaponStat(WEAPONSTAT_TYPE_DAMAGE,BOW_TYPE,LVTier_Bow,PlayerData.LV_BowDamage)(0)
            set CriticalDamage_Bow = CalculeWeaponStat(WEAPONSTAT_TYPE_CRITICALDAMAGE,BOW_TYPE,LVTier_Sword,PlayerData.LV_BowCriticalDamage)(0)
            set CriticalProb_Bow = CalculeWeaponStat(WEAPONSTAT_TYPE_CRITICALPROB,BOW_TYPE,LVTier_Sword,PlayerData.LV_BowCriticalProb)(0)
            #set Reward_Bow = CalculeWeaponStat(WEAPONSTAT_TYPE_REWARD,SWORD_TYPE,LVTier_Bow,PlayerData.LV_BowReward)(0)


    MakeWeaponsUpgradesCanvas(): canvas=
        Canvas := canvas:
            Slots:=array:
                #Background image
                canvas_slot:
                    Anchors:=anchors{Minimum:=vector2{X:=0.5,Y:=0.5},Maximum:=vector2{X:=0.5,Y:=0.5}}
                    ZOrder:=0
                    Offsets:=margin{Left:=0.0,Top:=0.0,Right:=BackgroundImageSizeX,Bottom:=BackgroundImageSizeY} #PositionX, PositionY, SizeX, SizeY
                    Alignment:=vector2{X:=0.5,Y:=0.5}
                    Widget:=texture_block:
                        DefaultImage:=Images.Img_MainBackground
                        DefaultDesiredSize:=vector2{X:=BackgroundImageSizeX,Y:=BackgroundImageSizeY}
                #Weapon selected
                canvas_slot:
                    Anchors:=anchors{Minimum:=vector2{X:=0.5,Y:=0.5},Maximum:=vector2{X:=0.5,Y:=0.5}}
                    ZOrder:=0
                    Alignment:=vector2{X:=0.5,Y:=0.5}
                    Widget:=overlay:
                        Slots:=array:
                            overlay_slot:
                                Padding:=margin{Bottom:=425.0}
                                Widget:=WeaponSelected_Texture
                #Main overlay
                canvas_slot:
                    Anchors:=anchors{Minimum:=vector2{X:=0.5,Y:=0.5},Maximum:=vector2{X:=0.5,Y:=0.5}}
                    ZOrder:=0
                    Alignment:=vector2{X:=0.5,Y:=0.5}
                    Widget:=overlay:
                        Slots:=array:
                            overlay_slot:
                                HorizontalAlignment:=horizontal_alignment.Center
                                VerticalAlignment:=vertical_alignment.Center
                                Padding:=margin{Left:=0.0,Top:=10.0,Right:=0.0,Bottom:=0.0}
                                Widget:=stack_box:
                                    Orientation:=orientation.Vertical
                                        Slots:=array:
                                            stack_box_slot: #Sapphires per produc
                                                HorizontalAlignment:=horizontal_alignment.Right
                                                Widget:=stack_box:
                                                    Orientation:=orientation.Horizontal
                                                        Slots:=array:
                                                            stack_box_slot: #Current SAPPHIRES
                                                                Padding:=margin{Left:=InterColumnLeftPadding,Top:=InterColumnTopPadding,Right:=InterColumnRightPadding,Bottom:=InterColumnBottomPadding}
                                                                Widget:=overlay:
                                                                    Slots:=array:
                                                                        overlay_slot:
                                                                            Widget:=TotalWater_BackgroundTexture
                                                                        overlay_slot:
                                                                            HorizontalAlignment:=horizontal_alignment.Left
                                                                            VerticalAlignment:=vertical_alignment.Center 
                                                                            Padding:=margin{Left:=10.0,Top:=0.0,Right:=0.0,Bottom:=0.0}
                                                                            Widget:=texture_block{DefaultDesiredSize:=vector2{X:=IconSmallSizeX,Y:=IconSmallSizeY}, DefaultImage:=Images.Sapphire_Icon}
                                                                        overlay_slot: 
                                                                            HorizontalAlignment:=horizontal_alignment.Left
                                                                            VerticalAlignment:=vertical_alignment.Center 
                                                                            Padding:=margin{Left:=65.0,Top:=0.0,Right:=0.0,Bottom:=0.0}
                                                                            Widget:=TotalSapphires_TextBlock
                                                            stack_box_slot: #Current MUTANTMATERIAL
                                                                Padding:=margin{Left:=InterColumnLeftPadding,Top:=InterColumnTopPadding,Right:=InterColumnRightPadding,Bottom:=InterColumnBottomPadding}
                                                                Widget:=overlay:
                                                                    Slots:=array:
                                                                        overlay_slot:
                                                                            Widget:=TotalMutantMaterial_BackgroundTexture
                                                                        overlay_slot:
                                                                            HorizontalAlignment:=horizontal_alignment.Left
                                                                            VerticalAlignment:=vertical_alignment.Center 
                                                                            Padding:=margin{Left:=10.0,Top:=0.0,Right:=0.0,Bottom:=0.0}
                                                                            Widget:=texture_block{DefaultDesiredSize:=vector2{X:=IconSmallSizeX,Y:=IconSmallSizeY}, DefaultImage:=Images.Mutant_Material_Icon}
                                                                        overlay_slot: 
                                                                            HorizontalAlignment:=horizontal_alignment.Left
                                                                            VerticalAlignment:=vertical_alignment.Center 
                                                                            Padding:=margin{Left:=65.0,Top:=0.0,Right:=0.0,Bottom:=0.0}
                                                                            Widget:=TotalMutantMaterial_TextBlock


                                                                          
                                            stack_box_slot: #Weapon upgrades
                                                Widget:=stack_box:
                                                    Orientation:=orientation.Vertical
                                                        Slots:=array:
                                                            stack_box_slot: #Line 0: All titles
                                                                HorizontalAlignment:=horizontal_alignment.Right
                                                                Padding:=margin{Left:=RowLeftPadding,Top:=25.0,Right:=RowRightPadding,Bottom:=RowBottomPadding}
                                                                    Widget:=overlay:
                                                                        Slots:=array:
                                                                            overlay_slot: 
                                                                                HorizontalAlignment:=horizontal_alignment.Fill
                                                                                VerticalAlignment:=vertical_alignment.Center
                                                                                Widget:=texture_block{DefaultDesiredSize:=vector2{X:=0.0,Y:=LineSizeY},DefaultImage:=Images.Img_LineBackground}
                                                                            overlay_slot: 
                                                                                HorizontalAlignment:=horizontal_alignment.Right
                                                                                Widget:=stack_box:
                                                                                    Orientation:=orientation.Horizontal
                                                                                        Slots:=array:
                                                                                            stack_box_slot: 
                                                                                                Padding:=margin{Left:=InterColumnLeftPadding,Top:=InterColumnTopPadding,Right:=InterColumnRightPadding,Bottom:=InterColumnBottomPadding}
                                                                                                Widget:=overlay:
                                                                                                    Slots:=array:
                                                                                                        overlay_slot:
                                                                                                            Widget:=text_block{DefaultText:=BasicMessage("EVOLVE"), DefaultTextColor:=White, DefaultShadowOffset:=ShadowOffset, DefaultShadowOpacity := DefaultTextShadowOpacity}
                                                                                                        overlay_slot: 
                                                                                                            Widget:=texture_block{DefaultDesiredSize:=vector2{X:=200.0},DefaultImage:=Images.EmptyTexture}
                                                                                            stack_box_slot: 
                                                                                                Padding:=margin{Left:=InterColumnLeftPadding,Top:=InterColumnTopPadding,Right:=InterColumnRightPadding,Bottom:=InterColumnBottomPadding}
                                                                                                Widget:=overlay:
                                                                                                    Slots:=array:
                                                                                                        overlay_slot: 
                                                                                                            Widget:=texture_block{DefaultDesiredSize:=vector2{X:=BUColumn6SizeX,Y:=BUColumn6SizeY},DefaultImage:=Images.EmptyTexture}
                                                                                                        overlay_slot: 
                                                                                                            HorizontalAlignment:=horizontal_alignment.Left
                                                                                                            VerticalAlignment:=vertical_alignment.Center 
                                                                                                            Padding:=margin{Left:=65.0,Top:=0.0,Right:=0.0,Bottom:=0.0}
                                                                                                            Widget:=PPU_Row1_UpgradeCost_TextBlock 
                                                                                                        overlay_slot: 
                                                                                                            HorizontalAlignment:=horizontal_alignment.Left
                                                                                                            VerticalAlignment:=vertical_alignment.Center 
                                                                                                            Padding:=margin{Left:=10.0,Top:=0.0,Right:=0.0,Bottom:=0.0}
                                                                                                            Widget:=texture_block{DefaultDesiredSize:=vector2{X:=IconSmallSizeX,Y:=IconSmallSizeY}, DefaultImage:=Images.Mutant_Material_Icon}
                                                                                                        overlay_slot:
                                                                                                            HorizontalAlignment:=horizontal_alignment.Fill
                                                                                                            VerticalAlignment:=vertical_alignment.Center  
                                                                                                            Widget:=PPU_Row1Button
                                                            stack_box_slot: #Line 1: Title
                                                                Padding:=margin{Left:=RowLeftPadding,Top:=RowTopPadding,Right:=RowRightPadding,Bottom:=RowBottomPadding}
                                                                HorizontalAlignment:=horizontal_alignment.Fill
                                                                VerticalAlignment:=vertical_alignment.Bottom
                                                                Widget:=overlay:
                                                                    Slots:=array:
                                                                        overlay_slot: 
                                                                            HorizontalAlignment:=horizontal_alignment.Fill
                                                                            VerticalAlignment:=vertical_alignment.Bottom
                                                                            Widget:=texture_block{DefaultDesiredSize:=vector2{X:=HorizontalLineSizeX,Y:=HorizontalLineSizeY},DefaultImage:=Images.Img_HorizontalLine}
                                                                        overlay_slot: 
                                                                            HorizontalAlignment:=horizontal_alignment.Center
                                                                            VerticalAlignment:=vertical_alignment.Center 
                                                                            Widget:=TitleTextBlock
                                                                        overlay_slot: 
                                                                            HorizontalAlignment:=horizontal_alignment.Center
                                                                            VerticalAlignment:=vertical_alignment.Center 
                                                                            Widget:=texture_block{DefaultDesiredSize:=vector2{X:=10.0,Y:=LineSizeY},DefaultImage:=Images.EmptyTexture}
                                                            stack_box_slot: #Line 2: All titles
                                                                Padding:=margin{Left:=RowLeftPadding,Top:=RowTopPadding,Right:=RowRightPadding,Bottom:=RowBottomPadding}
                                                                    Widget:=overlay:
                                                                        Slots:=array:
                                                                            overlay_slot: 
                                                                                HorizontalAlignment:=horizontal_alignment.Fill
                                                                                VerticalAlignment:=vertical_alignment.Center
                                                                                Widget:=texture_block{DefaultDesiredSize:=vector2{X:=0.0,Y:=LineSizeY},DefaultImage:=Images.Img_LineBackground}
                                                                            overlay_slot: 
                                                                                Widget:=stack_box:
                                                                                    Orientation:=orientation.Horizontal
                                                                                        Slots:=array:
                                                                                            stack_box_slot: 
                                                                                                Padding:=margin{Left:=InterColumnLeftPadding,Top:=InterColumnTopPadding,Right:=InterColumnRightPadding,Bottom:=InterColumnBottomPadding}
                                                                                                Widget:=overlay:
                                                                                                    Slots:=array:
                                                                                                        overlay_slot:
                                                                                                            Widget:=text_block{DefaultText:=BasicMessage("Stat"), DefaultTextColor:=White, DefaultShadowOffset:=ShadowOffset, DefaultShadowOpacity := DefaultTextShadowOpacity}
                                                                                                        overlay_slot: 
                                                                                                            Widget:=texture_block{DefaultDesiredSize:=vector2{X:=BUColumn2SizeX,Y:=BUColumn2SizeY},DefaultImage:=Images.EmptyTexture}
                                                                                        
                                                                                            stack_box_slot: 
                                                                                                Padding:=margin{Left:=InterColumnLeftPadding,Top:=InterColumnTopPadding,Right:=InterColumnRightPadding,Bottom:=InterColumnBottomPadding}
                                                                                                Widget:=overlay:
                                                                                                    Slots:=array:
                                                                                                        overlay_slot:
                                                                                                            Widget:=text_block{DefaultText:=BasicMessage("Current"), DefaultTextColor:=White, DefaultShadowOffset:=ShadowOffset, DefaultShadowOpacity := DefaultTextShadowOpacity}
                                                                                                        overlay_slot: 
                                                                                                            Widget:=texture_block{DefaultDesiredSize:=vector2{X:=BUColumn4SizeX,Y:=BUColumn4SizeY},DefaultImage:=Images.EmptyTexture}
                                                                                            stack_box_slot: 
                                                                                                Padding:=margin{Left:=InterColumnLeftPadding,Top:=InterColumnTopPadding,Right:=InterColumnRightPadding,Bottom:=InterColumnBottomPadding}
                                                                                                Widget:=overlay:
                                                                                                    Slots:=array:
                                                                                                        overlay_slot:
                                                                                                            Widget:=text_block{DefaultText:=BasicMessage("Upgrade"), DefaultTextColor:=White, DefaultShadowOffset:=ShadowOffset, DefaultShadowOpacity := DefaultTextShadowOpacity}
                                                                                                        overlay_slot: 
                                                                                                            Widget:=texture_block{DefaultDesiredSize:=vector2{X:=BUColumn5SizeX,Y:=BUColumn5SizeY},DefaultImage:=Images.EmptyTexture}
                                                                                            stack_box_slot: 
                                                                                                Padding:=margin{Left:=InterColumnLeftPadding,Top:=InterColumnTopPadding,Right:=InterColumnRightPadding,Bottom:=InterColumnBottomPadding}
                                                                                                Widget:=overlay:
                                                                                                    Slots:=array:
                                                                                                        overlay_slot:
                                                                                                            Widget:=text_block{DefaultText:=BasicMessage("Upgrade cost"), DefaultTextColor:=White, DefaultShadowOffset:=ShadowOffset, DefaultShadowOpacity := DefaultTextShadowOpacity}
                                                                                                        overlay_slot: 
                                                                                                            Widget:=texture_block{DefaultDesiredSize:=vector2{X:=BUColumn6SizeX,Y:=BUColumn6SizeY},DefaultImage:=Images.EmptyTexture}


                                                            stack_box_slot: #Line 3: 
                                                                Padding:=margin{Left:=RowLeftPadding,Top:=RowTopPadding,Right:=RowRightPadding,Bottom:=RowBottomPadding}
                                                                    Widget:=overlay:
                                                                        Slots:=array:
                                                                            overlay_slot: 
                                                                                HorizontalAlignment:=horizontal_alignment.Fill
                                                                                VerticalAlignment:=vertical_alignment.Center
                                                                                Widget:=texture_block{DefaultDesiredSize:=vector2{X:=0.0,Y:=LineSizeY},DefaultImage:=Images.Img_LineBackground}
                                                                            overlay_slot: 
                                                                                Widget:=stack_box:
                                                                                    Orientation:=orientation.Horizontal
                                                                                        Slots:=array:
                                                                                            stack_box_slot: 
                                                                                                Padding:=margin{Left:=InterColumnLeftPadding,Top:=InterColumnTopPadding,Right:=InterColumnRightPadding,Bottom:=InterColumnBottomPadding}
                                                                                                Widget:=overlay:
                                                                                                    Slots:=array:
                                                                                                        overlay_slot:
                                                                                                            Widget:=text_block{DefaultText:=BasicMessage("Damage"), DefaultTextColor:=White, DefaultShadowOffset:=ShadowOffset, DefaultShadowOpacity := DefaultTextShadowOpacity}
                                                                                                        overlay_slot: 
                                                                                                            Widget:=texture_block{DefaultDesiredSize:=vector2{X:=BUColumn2SizeX,Y:=BUColumn2SizeY},DefaultImage:=Images.EmptyTexture}

                                                                                           
                                                                                            stack_box_slot: 
                                                                                                Padding:=margin{Left:=InterColumnLeftPadding,Top:=InterColumnTopPadding,Right:=InterColumnRightPadding,Bottom:=InterColumnBottomPadding}
                                                                                                Widget:=overlay:
                                                                                                    Slots:=array:
                                                                                                        overlay_slot:
                                                                                                            Widget:=PPU_Row2_Current_TextBlock
                                                                                                        overlay_slot: 
                                                                                                            Widget:=texture_block{DefaultDesiredSize:=vector2{X:=BUColumn4SizeX,Y:=BUColumn4SizeY},DefaultImage:=Images.EmptyTexture}
                                                                                            stack_box_slot: 
                                                                                                Padding:=margin{Left:=InterColumnLeftPadding,Top:=InterColumnTopPadding,Right:=InterColumnRightPadding,Bottom:=InterColumnBottomPadding}
                                                                                                Widget:=overlay:
                                                                                                    Slots:=array:
                                                                                                        overlay_slot:
                                                                                                            Widget:=PPU_Row2_Upgrade_TextBlock
                                                                                                        overlay_slot: 
                                                                                                            Widget:=texture_block{DefaultDesiredSize:=vector2{X:=BUColumn5SizeX,Y:=BUColumn5SizeY},DefaultImage:=Images.EmptyTexture}
                                                                                            stack_box_slot: 
                                                                                                Padding:=margin{Left:=InterColumnLeftPadding,Top:=InterColumnTopPadding,Right:=InterColumnRightPadding,Bottom:=InterColumnBottomPadding}
                                                                                                Widget:=overlay:
                                                                                                    Slots:=array:
                                                                                                        overlay_slot: 
                                                                                                            Widget:=texture_block{DefaultDesiredSize:=vector2{X:=BUColumn6SizeX,Y:=BUColumn6SizeY},DefaultImage:=Images.EmptyTexture}
                                                                                                        overlay_slot: 
                                                                                                            HorizontalAlignment:=horizontal_alignment.Left
                                                                                                            VerticalAlignment:=vertical_alignment.Center 
                                                                                                            Padding:=margin{Left:=65.0,Top:=0.0,Right:=0.0,Bottom:=0.0}
                                                                                                            Widget:=PPU_Row2_UpgradeCost_TextBlock 
                                                                                                        overlay_slot: 
                                                                                                            HorizontalAlignment:=horizontal_alignment.Left
                                                                                                            VerticalAlignment:=vertical_alignment.Center 
                                                                                                            Padding:=margin{Left:=10.0,Top:=0.0,Right:=0.0,Bottom:=0.0}
                                                                                                            Widget:=texture_block{DefaultDesiredSize:=vector2{X:=IconSmallSizeX,Y:=IconSmallSizeY}, DefaultImage:=Images.Sapphire_Icon}
                                                                                                        overlay_slot:
                                                                                                            HorizontalAlignment:=horizontal_alignment.Fill
                                                                                                            VerticalAlignment:=vertical_alignment.Center  
                                                                                                            Widget:=PPU_Row2Button
                                                                                                        overlay_slot:
                                                                                                            Widget:=PPU_Row2_UpgradeCost_TextBlock
                                                                                            

                                                            stack_box_slot: #Line 4: All titles
                                                                Padding:=margin{Left:=RowLeftPadding,Top:=RowTopPadding,Right:=RowRightPadding,Bottom:=RowBottomPadding}
                                                                    Widget:=overlay:
                                                                        Slots:=array:
                                                                            overlay_slot: 
                                                                                HorizontalAlignment:=horizontal_alignment.Fill
                                                                                VerticalAlignment:=vertical_alignment.Center
                                                                                Widget:=texture_block{DefaultDesiredSize:=vector2{X:=0.0,Y:=LineSizeY},DefaultImage:=Images.Img_LineBackground}
                                                                            overlay_slot: 
                                                                                Widget:=stack_box:
                                                                                    Orientation:=orientation.Horizontal
                                                                                        Slots:=array:
                                                                                            stack_box_slot: 
                                                                                                Padding:=margin{Left:=InterColumnLeftPadding,Top:=InterColumnTopPadding,Right:=InterColumnRightPadding,Bottom:=InterColumnBottomPadding}
                                                                                                Widget:=overlay:
                                                                                                    Slots:=array:
                                                                                                        overlay_slot:
                                                                                                            Widget:=text_block{DefaultText:=BasicMessage("Critical probability"), DefaultTextColor:=White, DefaultShadowOffset:=ShadowOffset, DefaultShadowOpacity := DefaultTextShadowOpacity}
                                                                                                        overlay_slot: 
                                                                                                            Widget:=texture_block{DefaultDesiredSize:=vector2{X:=BUColumn2SizeX,Y:=BUColumn2SizeY},DefaultImage:=Images.EmptyTexture}
                                                                                            
                                                                                            stack_box_slot: 
                                                                                                Padding:=margin{Left:=InterColumnLeftPadding,Top:=InterColumnTopPadding,Right:=InterColumnRightPadding,Bottom:=InterColumnBottomPadding}
                                                                                                Widget:=overlay:
                                                                                                    Slots:=array:
                                                                                                        overlay_slot:
                                                                                                            Widget:=PPU_Row3_Current_TextBlock
                                                                                                        overlay_slot: 
                                                                                                            Widget:=texture_block{DefaultDesiredSize:=vector2{X:=BUColumn4SizeX,Y:=BUColumn4SizeY},DefaultImage:=Images.EmptyTexture}
                                                                                            stack_box_slot: 
                                                                                                Padding:=margin{Left:=InterColumnLeftPadding,Top:=InterColumnTopPadding,Right:=InterColumnRightPadding,Bottom:=InterColumnBottomPadding}
                                                                                                Widget:=overlay:
                                                                                                    Slots:=array:
                                                                                                        overlay_slot:
                                                                                                            Widget:=PPU_Row3_Upgrade_TextBlock
                                                                                                        overlay_slot: 
                                                                                                            Widget:=texture_block{DefaultDesiredSize:=vector2{X:=BUColumn5SizeX,Y:=BUColumn5SizeY},DefaultImage:=Images.EmptyTexture}
                                                                                            stack_box_slot: 
                                                                                                Padding:=margin{Left:=InterColumnLeftPadding,Top:=InterColumnTopPadding,Right:=InterColumnRightPadding,Bottom:=InterColumnBottomPadding}
                                                                                                Widget:=overlay:
                                                                                                    Slots:=array:
                                                                                                        overlay_slot: 
                                                                                                            Widget:=texture_block{DefaultDesiredSize:=vector2{X:=BUColumn6SizeX,Y:=BUColumn6SizeY},DefaultImage:=Images.EmptyTexture}
                                                                                                        overlay_slot: 
                                                                                                            HorizontalAlignment:=horizontal_alignment.Left
                                                                                                            VerticalAlignment:=vertical_alignment.Center 
                                                                                                            Padding:=margin{Left:=65.0,Top:=0.0,Right:=0.0,Bottom:=0.0}
                                                                                                            Widget:=PPU_Row3_UpgradeCost_TextBlock 
                                                                                                        overlay_slot: 
                                                                                                            HorizontalAlignment:=horizontal_alignment.Left
                                                                                                            VerticalAlignment:=vertical_alignment.Center 
                                                                                                            Padding:=margin{Left:=10.0,Top:=0.0,Right:=0.0,Bottom:=0.0}
                                                                                                            Widget:=texture_block{DefaultDesiredSize:=vector2{X:=IconSmallSizeX,Y:=IconSmallSizeY}, DefaultImage:=Images.Sapphire_Icon}
                                                                                                        overlay_slot:
                                                                                                            HorizontalAlignment:=horizontal_alignment.Fill
                                                                                                            VerticalAlignment:=vertical_alignment.Center  
                                                                                                            Widget:=PPU_Row3Button
                                                                                                        overlay_slot:
                                                                                                            Widget:=PPU_Row3_UpgradeCost_TextBlock
                                                                            

                                                            stack_box_slot: #Line 5: All titles
                                                                Padding:=margin{Left:=RowLeftPadding,Top:=RowTopPadding,Right:=RowRightPadding,Bottom:=RowBottomPadding}
                                                                    Widget:=overlay:
                                                                        Slots:=array:
                                                                            overlay_slot: 
                                                                                HorizontalAlignment:=horizontal_alignment.Fill
                                                                                VerticalAlignment:=vertical_alignment.Center
                                                                                Widget:=texture_block{DefaultDesiredSize:=vector2{X:=0.0,Y:=LineSizeY},DefaultImage:=Images.Img_LineBackground}
                                                                            overlay_slot: 
                                                                                Widget:=stack_box:
                                                                                    Orientation:=orientation.Horizontal
                                                                                        Slots:=array:
                                                                                            stack_box_slot: 
                                                                                                Padding:=margin{Left:=InterColumnLeftPadding,Top:=InterColumnTopPadding,Right:=InterColumnRightPadding,Bottom:=InterColumnBottomPadding}
                                                                                                Widget:=overlay:
                                                                                                    Slots:=array:
                                                                                                        overlay_slot:
                                                                                                            Widget:=text_block{DefaultText:=BasicMessage("Critical damage"), DefaultTextColor:=White, DefaultShadowOffset:=ShadowOffset, DefaultShadowOpacity := DefaultTextShadowOpacity}
                                                                                                        overlay_slot: 
                                                                                                            Widget:=texture_block{DefaultDesiredSize:=vector2{X:=BUColumn2SizeX,Y:=BUColumn2SizeY},DefaultImage:=Images.EmptyTexture}
                                                                                            
                                                                                            stack_box_slot: 
                                                                                                Padding:=margin{Left:=InterColumnLeftPadding,Top:=InterColumnTopPadding,Right:=InterColumnRightPadding,Bottom:=InterColumnBottomPadding}
                                                                                                Widget:=overlay:
                                                                                                    Slots:=array:
                                                                                                        overlay_slot:
                                                                                                            Widget:=PPU_Row4_Current_TextBlock
                                                                                                        overlay_slot: 
                                                                                                            Widget:=texture_block{DefaultDesiredSize:=vector2{X:=BUColumn4SizeX,Y:=BUColumn4SizeY},DefaultImage:=Images.EmptyTexture}
                                                                                            stack_box_slot: 
                                                                                                Padding:=margin{Left:=InterColumnLeftPadding,Top:=InterColumnTopPadding,Right:=InterColumnRightPadding,Bottom:=InterColumnBottomPadding}
                                                                                                Widget:=overlay:
                                                                                                    Slots:=array:
                                                                                                        overlay_slot:
                                                                                                            Widget:=PPU_Row4_Upgrade_TextBlock
                                                                                                        overlay_slot: 
                                                                                                            Widget:=texture_block{DefaultDesiredSize:=vector2{X:=BUColumn5SizeX,Y:=BUColumn5SizeY},DefaultImage:=Images.EmptyTexture}
                                                                                            stack_box_slot: 
                                                                                                Padding:=margin{Left:=InterColumnLeftPadding,Top:=InterColumnTopPadding,Right:=InterColumnRightPadding,Bottom:=InterColumnBottomPadding}
                                                                                                Widget:=overlay:
                                                                                                    Slots:=array:
                                                                                                        overlay_slot: 
                                                                                                            Widget:=texture_block{DefaultDesiredSize:=vector2{X:=BUColumn6SizeX,Y:=BUColumn6SizeY},DefaultImage:=Images.EmptyTexture}
                                                                                                        overlay_slot: 
                                                                                                            HorizontalAlignment:=horizontal_alignment.Left
                                                                                                            VerticalAlignment:=vertical_alignment.Center 
                                                                                                            Padding:=margin{Left:=65.0,Top:=0.0,Right:=0.0,Bottom:=0.0}
                                                                                                            Widget:=PPU_Row4_UpgradeCost_TextBlock 
                                                                                                        overlay_slot: 
                                                                                                            HorizontalAlignment:=horizontal_alignment.Left
                                                                                                            VerticalAlignment:=vertical_alignment.Center 
                                                                                                            Padding:=margin{Left:=10.0,Top:=0.0,Right:=0.0,Bottom:=0.0}
                                                                                                            Widget:=texture_block{DefaultDesiredSize:=vector2{X:=IconSmallSizeX,Y:=IconSmallSizeY}, DefaultImage:=Images.Sapphire_Icon}
                                                                                                        overlay_slot:
                                                                                                            HorizontalAlignment:=horizontal_alignment.Fill
                                                                                                            VerticalAlignment:=vertical_alignment.Center  
                                                                                                            Widget:=PPU_Row4Button
                                                                                                        overlay_slot:
                                                                                                            Widget:=PPU_Row4_UpgradeCost_TextBlock
                                                                                                
<#
                                                            stack_box_slot: #Line 6: All titles
                                                                Padding:=margin{Left:=RowLeftPadding,Top:=RowTopPadding,Right:=RowRightPadding,Bottom:=RowBottomPadding}
                                                                    Widget:=overlay:
                                                                        Slots:=array:
                                                                            overlay_slot: 
                                                                                HorizontalAlignment:=horizontal_alignment.Fill
                                                                                VerticalAlignment:=vertical_alignment.Center
                                                                                Widget:=texture_block{DefaultDesiredSize:=vector2{X:=0.0,Y:=LineSizeY},DefaultImage:=Images.Img_LineBackground}
                                                                            overlay_slot: 
                                                                                Widget:=stack_box:
                                                                                    Orientation:=orientation.Horizontal
                                                                                        Slots:=array:
                                                                                            stack_box_slot: 
                                                                                                Padding:=margin{Left:=InterColumnLeftPadding,Top:=InterColumnTopPadding,Right:=InterColumnRightPadding,Bottom:=InterColumnBottomPadding}
                                                                                                Widget:=overlay:
                                                                                                    Slots:=array:
                                                                                                        overlay_slot:
                                                                                                            Widget:=text_block{DefaultText:=BasicMessage("Mutant gather"), DefaultTextColor:=White, DefaultShadowOffset:=ShadowOffset, DefaultShadowOpacity := DefaultTextShadowOpacity}
                                                                                                        overlay_slot: 
                                                                                                            Widget:=texture_block{DefaultDesiredSize:=vector2{X:=BUColumn2SizeX,Y:=BUColumn2SizeY},DefaultImage:=Images.EmptyTexture}
                                                                                        
                                                                                            stack_box_slot: 
                                                                                                Padding:=margin{Left:=InterColumnLeftPadding,Top:=InterColumnTopPadding,Right:=InterColumnRightPadding,Bottom:=InterColumnBottomPadding}
                                                                                                Widget:=overlay:
                                                                                                    Slots:=array:
                                                                                                        overlay_slot:
                                                                                                            Widget:=PPU_Row5_Current_TextBlock
                                                                                                        overlay_slot: 
                                                                                                            Widget:=texture_block{DefaultDesiredSize:=vector2{X:=BUColumn4SizeX,Y:=BUColumn4SizeY},DefaultImage:=Images.EmptyTexture}
                                                                                            stack_box_slot: 
                                                                                                Padding:=margin{Left:=InterColumnLeftPadding,Top:=InterColumnTopPadding,Right:=InterColumnRightPadding,Bottom:=InterColumnBottomPadding}
                                                                                                Widget:=overlay:
                                                                                                    Slots:=array:
                                                                                                        overlay_slot:
                                                                                                            Widget:=PPU_Row5_Upgrade_TextBlock
                                                                                                        overlay_slot: 
                                                                                                            Widget:=texture_block{DefaultDesiredSize:=vector2{X:=BUColumn5SizeX,Y:=BUColumn5SizeY},DefaultImage:=Images.EmptyTexture}
                                                                                            stack_box_slot: 
                                                                                                Padding:=margin{Left:=InterColumnLeftPadding,Top:=InterColumnTopPadding,Right:=InterColumnRightPadding,Bottom:=InterColumnBottomPadding}
                                                                                                Widget:=overlay:
                                                                                                    Slots:=array:
                                                                                                        overlay_slot: 
                                                                                                            Widget:=texture_block{DefaultDesiredSize:=vector2{X:=BUColumn6SizeX,Y:=BUColumn6SizeY},DefaultImage:=Images.EmptyTexture}
                                                                                                        overlay_slot: 
                                                                                                            HorizontalAlignment:=horizontal_alignment.Left
                                                                                                            VerticalAlignment:=vertical_alignment.Center 
                                                                                                            Padding:=margin{Left:=65.0,Top:=0.0,Right:=0.0,Bottom:=0.0}
                                                                                                            Widget:=PPU_Row5_UpgradeCost_TextBlock 
                                                                                                        overlay_slot: 
                                                                                                            HorizontalAlignment:=horizontal_alignment.Left
                                                                                                            VerticalAlignment:=vertical_alignment.Center 
                                                                                                            Padding:=margin{Left:=10.0,Top:=0.0,Right:=0.0,Bottom:=0.0}
                                                                                                            Widget:=texture_block{DefaultDesiredSize:=vector2{X:=IconSmallSizeX,Y:=IconSmallSizeY}, DefaultImage:=Images.Sapphire_Icon}
                                                                                                        overlay_slot:
                                                                                                            HorizontalAlignment:=horizontal_alignment.Fill
                                                                                                            VerticalAlignment:=vertical_alignment.Center  
                                                                                                            Widget:=PPU_Row5Button
                                                                                                        overlay_slot:
                                                                                                            Widget:=PPU_Row5_UpgradeCost_TextBlock   #>
                                                                                                            
                                                                                                           
                                                            <#
                                                            stack_box_slot: #Line 8: Navigation buttons
                                                                Padding:=margin{Left:=RowLeftPadding,Top:=RowTopPadding,Right:=RowRightPadding,Bottom:=RowBottomPadding}
                                                                    Widget:=overlay:
                                                                        Slots:=array:
                                                                            overlay_slot: 
                                                                                HorizontalAlignment:=horizontal_alignment.Center
                                                                                VerticalAlignment:=vertical_alignment.Center
                                                                                Widget:=stack_box:
                                                                                    Orientation:=orientation.Horizontal
                                                                                        Slots:=array:
                                                                                            stack_box_slot: 
                                                                                                Padding:=margin{Left:=InterColumnLeftPadding,Top:=InterColumnTopPadding,Right:=InterColumnRightPadding,Bottom:=InterColumnBottomPadding}
                                                                                                Widget:=overlay:
                                                                                                    Slots:=array:
                                                                                                        overlay_slot: 
                                                                                                            Widget:=texture_block{DefaultDesiredSize:=vector2{X:=BigCloseButtonSizeX-500.0},DefaultImage:=Images.EmptyTexture}
                                                                                                        overlay_slot: 
                                                                                                            HorizontalAlignment:=horizontal_alignment.Fill
                                                                                                            Padding:=margin{Left:=InterColumnLeftPadding,Top:=InterColumnTopPadding,Right:=InterColumnRightPadding,Bottom:=InterColumnBottomPadding}
                                                                                                            Widget:=NavigateLeftButton
                                                                                            stack_box_slot: 
                                                                                                Padding:=margin{Left:=InterColumnLeftPadding,Top:=InterColumnTopPadding,Right:=InterColumnRightPadding,Bottom:=InterColumnBottomPadding}
                                                                                                Widget:=overlay:
                                                                                                    Slots:=array:
                                                                                                        overlay_slot: 
                                                                                                            Widget:=texture_block{DefaultDesiredSize:=vector2{X:=BigCloseButtonSizeX-500.0},DefaultImage:=Images.EmptyTexture}
                                                                                                        overlay_slot: 
                                                                                                            HorizontalAlignment:=horizontal_alignment.Fill
                                                                                                            Padding:=margin{Left:=InterColumnLeftPadding,Top:=InterColumnTopPadding,Right:=InterColumnRightPadding,Bottom:=InterColumnBottomPadding}
                                                                                                            Widget:=NavigateRightButton#>

                                                            stack_box_slot: #Line 9: Close button
                                                                Padding:=margin{Left:=RowLeftPadding,Top:=RowTopPadding,Right:=RowRightPadding,Bottom:=RowBottomPadding}
                                                                    Widget:=overlay:
                                                                        Slots:=array:
                                                                            overlay_slot: 
                                                                                HorizontalAlignment:=horizontal_alignment.Center
                                                                                VerticalAlignment:=vertical_alignment.Center
                                                                                Widget:=stack_box:
                                                                                    Orientation:=orientation.Horizontal
                                                                                        Slots:=array:
                                                                                            stack_box_slot: 
                                                                                                Padding:=margin{Left:=InterColumnLeftPadding,Top:=InterColumnTopPadding,Right:=InterColumnRightPadding,Bottom:=InterColumnBottomPadding}
                                                                                                Widget:=overlay:
                                                                                                    Slots:=array:
                                                                                                        overlay_slot: 
                                                                                                            Widget:=texture_block{DefaultDesiredSize:=vector2{X:=BigCloseButtonSizeX,Y:=BigCloseButtonSizeY},DefaultImage:=Images.EmptyTexture}
                                                                                                        overlay_slot: 
                                                                                                            HorizontalAlignment:=horizontal_alignment.Fill
                                                                                                            Padding:=margin{Left:=InterColumnLeftPadding,Top:=InterColumnTopPadding,Right:=InterColumnRightPadding,Bottom:=InterColumnBottomPadding}
                                                                                                            Widget:=CloseButton
                                                            
MakeWeaponsUpgradeUI<constructor>(
    InBakeryMachinesManager : bakery_machines_manager)<transacts> := weapons_upgrades_ui :
        BakeryMachinesManager := InBakeryMachinesManager
                                                                                                               