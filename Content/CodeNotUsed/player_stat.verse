using { /Fortnite.com/Devices }
using { /Verse.org/Simulation }
using { /UnrealEngine.com/Temporary/Diagnostics }
# A persistable stat that persists across play sessions.
player_stat := class<final><persistable>:
    # The current value of the player's stat.
    CurrentValue:int = 0

    # The highest value achieved so far.
    HighestValue:int = 0

    # Returns either the NewValue or HighestValue,
    # whichever is higher.
    SetHighestValue(NewValue:int)<transacts>:int=
        if (NewValue > HighestValue):
            NewValue
        else:
            HighestValue

# Creates a new player_stat class with the given NewValue and a source player_stat.
# The HighestValue is set to either the NewValue or the HighestValue
# from the source player_stat.
MakeUpdatedPlayerStat<constructor>(OldStats:player_stat, NewValue:int)<transacts> := player_stat:
    CurrentValue := NewValue
    HighestValue := OldStats.SetHighestValue(NewValue)