player_stats_table := class<final><persistable>:
    #Version
    Version:int = 0
    Wins:player_stat = player_stat{}
    Losses:player_stat = player_stat{}
    Score:player_stat = player_stat{}
    Record:player_stat = player_stat{}
    Fame:player_stat = player_stat{}
    Rank:player_stat = player_stat{}
    CompletedQuestCount:int = 0

    # Maps players to a table of their player stats.
    #var PlayerStatsMap : weak_map(player, player_stats_table) = map{}


MakePlayerStatsTable<constructor>(OldTable:player_stats_table)<transacts> := player_stats_table:
    Version := OldTable.Version
    Wins := OldTable.Wins
    Losses := OldTable.Losses
    Score := OldTable.Score
    Record := OldTable.Record
    Fame := OldTable.Fame
    Rank := OldTable.Rank
    CompletedQuestCount := OldTable.CompletedQuestCount

